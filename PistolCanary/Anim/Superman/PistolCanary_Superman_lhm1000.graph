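# Level Sequence for the PistolCanary "Superman" animation, animating the cinematic superhero NPC PistolCanary_CinematicNPC_Superhero_Main in front of a cine camera.
# The camera has a spawn track and constant default values for transform, focal length, manual focus distance and aperture.
# The NPC's Fortnite_3L_Face_CtrlRig track carries densely keyed, baked-style facial control curves, including eye-aim location and rotation, starting before frame zero.
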LevelSequence /PistolCanary/Anim/Superman/PistolCanary_Superman_lhm1000
bindings [
 CineCameraActor {
  id=#1
  possessed_class=CineCameraActor
  tracks [
   {
    type=MovieScene3DTransformTrack
    sections [{
      type=MovieScene3DTransformSection
      channels [
       Location.X {type=MovieSceneScriptingDoubleChannel, default=30.7391}
       Location.Y {type=MovieSceneScriptingDoubleChannel, default=435.5093}
       Location.Z {type=MovieSceneScriptingDoubleChannel, default=47.5476}
       Rotation.X {type=MovieSceneScriptingDoubleChannel, default=0}
       Rotation.Y {type=MovieSceneScriptingDoubleChannel, default=5.0001}
       Rotation.Z {type=MovieSceneScriptingDoubleChannel, default=-93.8}
       Scale.X {type=MovieSceneScriptingDoubleChannel, default=1}
       Scale.Y {type=MovieSceneScriptingDoubleChannel, default=1}
       Scale.Z {type=MovieSceneScriptingDoubleChannel, default=1}
      ]
     }]
   }
   {type=MovieSceneSpawnTrack, sections [{type=MovieSceneSpawnSection, channels [None {type=MovieSceneScriptingBoolChannel, default=true}]}]}
  ]
 }
 CameraComponent {
  id=#2
  parent_id=#1
  possessed_class=CineCameraComponent
  tracks [
   {
    type=MovieSceneFloatTrack
    display_name=Current Focal Length
    sections [{type=MovieSceneFloatSection, channels [None {type=MovieSceneScriptingFloatChannel, default=50}]}]
   }
   {
    type=MovieSceneFloatTrack
    display_name="Manual Focus Distance (Focus Settings)"
    sections [{type=MovieSceneFloatSection, channels [None {type=MovieSceneScriptingFloatChannel, default=100000}]}]
   }
   {
    type=MovieSceneFloatTrack
    display_name=Current Aperture
    sections [{type=MovieSceneFloatSection, channels [None {type=MovieSceneScriptingFloatChannel, default=10}]}]
   }
  ]
 }
 PistolCanary_CinematicNPC_Superhero_Main {
  id=#3
  possessed_class=PistolCanary_CinematicNPC_Superhero_Main_C
  tracks [
   {
    type=MovieSceneControlRigParameterTrack
    display_name=Fortnite_3L_Face_CtrlRig
    sections [
     {
      type=MovieSceneControlRigParameterSection
      channels [
       CTRL_C_eyesAim.Location.X {
        type=MovieSceneScriptingFloatChannel
        default=0
        thinned_from=492
        keys [[-262,7.2541],[-260,6.7031],[-258,6.8052],[-256,6.7499],[-254,6.4115],[-252,6.0574],[-250,5.9028],[-248,5.6641],[-244,5.6949],[-242,5.6075],[-240,5.4964],[-234,5.4871],[-232,5.4449],[-230,5.3766],[-228,5.1946],[-226,4.998],[-224,4.7174],[-222,4.5183],[-220,4.4873],[-218,4.6637],[-216,4.7831],[-214,4.8337],[-212,4.8422],[-210,4.8894],[-208,4.6972],[-206,4.4805],[-204,4.347],[-202,4.1694],[-200,3.9643],[-198,3.7798],[-196,3.5582],[-194,3.3724],[-192,3.2417],[-190,3.1967],[-188,3.0192],[-186,3.0087],[-184,3.0188],[-182,3.1376],[-180,3.224],[-176,3.6566],[-174,3.9322],[-172,4.1538],[-168,4.5445],[-166,4.6522],[-164,4.7975],[-160,4.8483],[-158,4.8624],[-156,4.9855],[-154,5.0627],[-152,4.9804],[-150,5.0998],[-148,4.8535],[-146,4.6618],[-144,4.5044],[-142,4.3805],[-140,4.4832],[-138,4.4394],[-136,4.4557],[-134,4.509],[-132,4.5348],[-130,4.6081],[-128,4.6082],[-126,4.7418],[-124,4.7255],[-120,4.7432],[-118,4.7115],[-116,4.8032],[-114,4.8],[-112,4.7835],[-110,4.8028],[-108,4.7639],[-106,4.7451],[-104,4.7716],[-102,4.8154],[-100,4.9252],[-98,4.9368],[-94,5.4155],[-92,5.4508],[-90,5.545],[-88,5.5649],[-86,5.648],[-84,5.7725],[-82,5.8117],[-80,5.6994],[-78,5.5254],[-76,5.4701],[-74,5.3296],[-72,5.2277],[-70,5.161],[-68,5.1968],[-66,5.1451],[-64,5.1169],[-62,5.1333],[-60,5.1668],[-56,5.1889],[-54,5.2132],[-52,5.2048],[-50,5.0784],[-48,5.0047],[-46,3.0534],[-44,2.0951],[-42,3.8083],[-40,4.281],[-38,4.285],[-36,4.3756],[-34,4.5626],[-32,4.7395],[-30,4.7643],[-28,4.7049],[-26,4.7095],[-24,4.7479],[-22,4.6288],[-20,4.7218],[-18,4.8026],[-16,4.7608],[-14,4.8453],[-12,4.8718],[-10,5.14],[-8,5.1777],[-6,5.2548],[-4,5.294],[-2,5.2913],[0,5.168],[2,5.1811],[4,5.0933],[6,5.1401],[8,5.1382],[12,5.2551],[14,5.3039],[18,5.2981],[20,5.3732],[22,5.4266],[26,5.4655],[28,5.5448],[30,5.5648],[32,5.5515],[34,5.5559],[36,5.6257],[38,5.648],[40,5.7027],[42,5.6173],[44,5.7046],[48,5.7435],[50,5.7272],[52,5.8051],[54,5.8314],[58,5.9156],[60,6.0145],[62,6.1702],[64,6.2303],[66,6.2483],[68,6.2839],[70,6.3036],[72,6.4318],[74,6.3306],[76,6.2489],[78,6.2378],[80,6.321],[82,6.3487],[84,6.2863],[86,6.3452],[88,6.3913],[90,6.3814],[92,6.3061],[94,6.1931],[96,5.9246],[98,5.726],[100,5.4317],[102,4.8284],[104,4.432],[106,3.4161],[108,3.2935],[110,2.86],[112,2.4164],[114,2.1776],[116,2.1518],[118,2.1847],[120,2.1922],[122,2.1447],[124,2.1122],[126,2.0985],[128,1.881],[130,1.881],[132,1.7114],[134,1.6906],[136,1.7518],[138,1.7588],[140,1.973],[142,2.1142],[144,2.2806],[146,2.757],[148,3.0268],[150,3.2032],[152,3.504],[154,3.8184],[156,3.8285],[158,3.9906],[160,4.202],[162,4.3349],[164,4.3675],[166,4.4551],[168,4.5669],[170,4.6588],[172,4.6617],[174,4.7025],[176,4.8367],[178,4.9117],[180,4.8382],[182,4.855],[184,4.7433],[190,4.7408],[192,4.6285],[194,4.6117],[196,4.5444],[198,4.4624],[200,4.4928],[202,4.5623],[204,4.5844],[206,4.6196],[208,4.7282],[210,4.7345],[212,4.7621],[214,4.8171],[216,4.91],[218,4.9839],[220,5.1512],[222,5.1552],[228,5.3059],[230,5.5514],[232,5.599],[234,5.6288],[236,5.6997],[238,5.826],[240,5.9157],[242,6.1521],[244,6.3218],[246,6.6539],[248,6.6473],[250,6.7648],[252,6.7872],[254,6.7611],[256,6.7732],[258,6.7254],[260,6.7094],[262,6.7502],[264,6.6109],[266,6.3568],[268,6.3042],[270,6.4753],[272,7.1458],[274,8.3237],[276,9.9316],[278,11.1469],[280,12.0405],[282,12.3543],[284,12.7351],[286,12.669],[290,12.4468],[292,12.4809],[294,12.3544],[296,12.4419],[298,12.3551],[300,12.3993],[304,12.2971],[308,12.2531],[310,12.1427],[312,12.1768],[316,12.0909],[318,11.9375],[320,12.1313],[322,12.2034],[324,12.1995],[326,12.2352],[328,12.4238],[330,12.3974],[332,12.4942],[334,12.7237],[336,13.0061],[340,12.9207],[344,12.2431],[346,11.7743],[348,11.3662],[350,10.9833],[352,10.716],[354,10.6297],[356,10.385],[358,10.3525],[360,10.2517],[362,10.2079],[366,10.1737],[368,10.1054],[370,9.9753],[372,9.7616],[374,9.1823],[376,9.0785],[378,8.6474],[380,7.9823],[382,7.5121],[384,7.5058],[386,8.0264],[388,10.4415],[390,10.8891],[392,10.9674],[394,10.2776],[396,8.6222],[398,6.9312],[400,6.5395],[402,5.7384],[404,3.767],[406,4.6276],[408,4.2231],[410,3.5113],[412,3.0208],[414,2.7129],[416,2.4483],[418,2.2465],[420,2.8384],[422,3.0575],[424,2.9294],[426,2.6042],[428,2.3234],[430,2.144],[432,2.0213],[434,1.925],[436,1.9585],[438,1.9717],[440,1.958],[442,2.0729],[444,2.1345],[448,2.2089],[450,2.2743],[452,2.4105],[454,2.5724],[456,2.8432],[458,3.1983],[460,3.613],[462,4.2728],[464,4.8886],[466,5.9701],[468,6.2981],[470,5.6367],[472,5.276],[474,3.6869],[476,1.7984],[478,3.4035],[480,2.8794],[482,0.9164],[484,0.5303],[486,0.2877],[488,0.9458],[490,1.7821],[492,2.7939],[494,2.8642],[496,3.0663],[498,3.5724],[500,4.0698],[502,4.1651],[504,4.5772],[506,5.1032],[508,5.342],[510,5.6424],[512,5.7717],[514,4.5402],[516,-0.0866],[518,-0.7139],[520,-7.1081],[522,-10.8392],[524,-12.2684],[526,-12.5837],[528,-12.5589],[530,-12.3909],[532,-12.2494],[534,-7.4316],[536,-3.2752],[538,4.8772],[540,6.4066],[542,9.8256],[544,9.2375],[546,8.5469],[548,7.2785],[550,6.6573],[552,6.6972],[554,6.6888],[556,6.5201],[558,6.4823],[560,6.4283],[562,6.4671],[564,6.4786],[566,6.5225],[568,6.5238],[570,6.4003],[572,6.5196],[578,6.5492],[580,6.5399],[582,6.3603],[584,5.9095],[586,5.6279],[588,5.8026],[590,6.0807],[592,6.0706],[594,6.0162],[596,4.8374],[598,4.0089],[600,2.6848],[602,2.604],[604,1.957],[606,1.8489],[608,1.523],[610,1.3336],[612,1.1106],[614,0.9642],[616,0.8598],[618,0.7971],[620,0.7768],[622,0.7384],[624,0.7247],[626,0.7296],[628,0.7741],[630,0.7583],[632,0.9044],[634,0.9417],[636,1.0559],[638,1.0837],[640,1.137],[642,1.0287],[644,1.0611],[646,1.096],[648,1.0178],[650,1.0277],[654,1.1735],[656,1.1237],[658,1.2136],[660,1.313],[662,1.3532],[664,1.3489],[666,1.5276],[668,1.6161],[670,1.6876],[672,1.7297],[674,1.719],[676,1.7866],[678,1.761],[680,1.842],[682,1.834],[684,1.8623],[686,1.7808],[688,1.836],[690,1.7769],[692,1.8243],[694,1.8162],[696,1.8548],[698,1.9713],[700,2.0187],[702,2.0495],[704,2.0876],[706,2.1477],[708,2.2169],[710,2.2532],[712,2.2496],[714,2.2837],[716,2.3656],[718,2.4392],[720,2.4392]]
       }
       CTRL_C_eyesAim.Location.Y {
        type=MovieSceneScriptingFloatChannel
        default=0
        thinned_from=492
        keys [[-262,1.0278],[-260,0.872],[-258,0.9045],[-256,0.8889],[-254,0.8127],[-252,0.734],[-250,0.7097],[-248,0.661],[-246,0.6725],[-244,0.6793],[-242,0.6514],[-240,0.6356],[-238,0.6248],[-236,0.6073],[-234,0.6047],[-232,0.5863],[-230,0.564],[-228,0.5098],[-226,0.4599],[-224,0.3986],[-222,0.3551],[-220,0.3428],[-218,0.3755],[-216,0.4021],[-214,0.414],[-212,0.4165],[-210,0.44],[-206,0.372],[-204,0.3443],[-202,0.3158],[-198,0.2645],[-196,0.2369],[-194,0.2161],[-192,0.202],[-190,0.1975],[-188,0.1823],[-186,0.1814],[-184,0.187],[-182,0.2063],[-180,0.2232],[-176,0.2938],[-174,0.347],[-172,0.3948],[-170,0.4278],[-168,0.4523],[-166,0.4691],[-164,0.4869],[-162,0.4798],[-160,0.4781],[-158,0.4799],[-156,0.4958],[-154,0.4953],[-152,0.4788],[-150,0.4992],[-148,0.4499],[-146,0.4167],[-144,0.3921],[-142,0.3717],[-140,0.4115],[-138,0.4076],[-136,0.411],[-134,0.4222],[-132,0.4197],[-130,0.419],[-128,0.42],[-126,0.4399],[-124,0.4382],[-122,0.4413],[-120,0.4433],[-118,0.432],[-116,0.4466],[-112,0.4474],[-110,0.4544],[-108,0.4492],[-106,0.4507],[-104,0.4585],[-102,0.4716],[-100,0.5116],[-98,0.5161],[-96,0.5664],[-94,0.6124],[-92,0.6216],[-90,0.6465],[-88,0.652],[-86,0.6721],[-84,0.6942],[-82,0.7013],[-80,0.678],[-78,0.635],[-76,0.6203],[-74,0.5726],[-72,0.5487],[-70,0.5311],[-68,0.5273],[-66,0.5096],[-64,0.4984],[-62,0.5004],[-60,0.5052],[-58,0.5051],[-56,0.504],[-54,0.5096],[-52,0.5113],[-50,0.4888],[-48,0.4742],[-46,0.1518],[-44,-0.0138],[-42,0.1931],[-40,0.2965],[-38,0.3323],[-36,0.3571],[-34,0.3908],[-32,0.4285],[-30,0.4405],[-28,0.4388],[-26,0.438],[-24,0.4464],[-22,0.4277],[-20,0.4452],[-18,0.4705],[-16,0.4615],[-14,0.4751],[-12,0.4772],[-10,0.5484],[-8,0.556],[-6,0.5738],[-4,0.5804],[-2,0.5715],[0,0.542],[2,0.543],[4,0.522],[6,0.5258],[8,0.523],[10,0.5322],[14,0.5572],[16,0.5566],[18,0.5582],[20,0.5766],[22,0.5895],[24,0.5943],[26,0.5949],[28,0.6147],[30,0.6262],[32,0.6269],[34,0.6233],[36,0.6388],[38,0.6443],[40,0.6585],[42,0.6448],[44,0.6641],[48,0.6715],[50,0.6705],[52,0.6862],[54,0.6886],[56,0.6988],[58,0.7122],[60,0.7449],[62,0.7885],[66,0.8075],[68,0.8249],[70,0.8295],[72,0.859],[74,0.8316],[76,0.8133],[78,0.8105],[80,0.8365],[82,0.8456],[84,0.836],[86,0.8605],[88,0.8788],[92,0.8719],[94,0.8366],[96,0.7727],[98,0.7189],[100,0.6517],[102,0.5219],[104,0.4592],[106,0.3074],[108,0.2933],[110,0.2325],[112,0.1697],[114,0.1243],[116,0.1014],[118,0.1079],[120,0.111],[122,0.1185],[124,0.1178],[126,0.1161],[128,0.0912],[130,0.0912],[132,0.0841],[134,0.0777],[136,0.0822],[138,0.0861],[140,0.0995],[142,0.1053],[144,0.1185],[146,0.1606],[148,0.1854],[150,0.2173],[152,0.2534],[154,0.3037],[156,0.3052],[158,0.33],[160,0.3569],[162,0.3914],[164,0.3869],[166,0.4051],[168,0.431],[170,0.4509],[172,0.4538],[174,0.4628],[176,0.4915],[178,0.5168],[180,0.5019],[182,0.5221],[184,0.5112],[186,0.5137],[188,0.524],[190,0.5281],[192,0.5037],[194,0.4993],[196,0.4817],[198,0.4721],[200,0.4923],[202,0.5103],[204,0.5142],[206,0.5304],[208,0.5625],[210,0.5644],[214,0.6127],[216,0.6531],[218,0.6886],[220,0.7351],[222,0.7549],[224,0.7797],[226,0.8236],[228,0.8366],[230,0.9242],[232,0.9656],[234,0.9967],[236,1.0523],[238,1.1132],[240,1.17],[242,1.2483],[244,1.2948],[246,1.4138],[248,1.4284],[250,1.4593],[252,1.4768],[256,1.4731],[258,1.452],[260,1.4552],[262,1.4635],[264,1.417],[266,1.3672],[268,1.4009],[270,1.4809],[272,1.6965],[274,2.1751],[276,2.8772],[278,3.401],[280,3.8954],[282,4.0704],[284,4.3078],[286,4.2368],[290,4.0462],[292,4.0507],[294,3.9864],[296,4.0309],[298,3.9968],[300,4.037],[302,4.018],[304,3.9814],[308,3.958],[310,3.8557],[312,3.8621],[314,3.83],[316,3.7775],[318,3.6786],[320,3.7595],[322,3.7871],[324,3.7609],[326,3.785],[328,3.8936],[330,3.8694],[332,3.9041],[334,4.0283],[336,4.1754],[338,4.1496],[340,4.1041],[342,3.8764],[344,3.666],[346,3.3739],[348,3.1228],[350,2.9021],[352,2.7607],[354,2.7157],[356,2.5844],[358,2.5516],[360,2.505],[362,2.4773],[364,2.4611],[366,2.4278],[368,2.3789],[370,2.305],[372,2.178],[374,1.8777],[376,1.826],[378,1.5998],[380,1.2947],[382,1.0912],[384,1.2055],[386,1.6628],[388,3.2984],[390,4.2028],[392,5.7973],[394,6.0754],[396,4.8491],[398,3.4018],[400,3.0698],[402,2.5235],[404,1.9387],[406,1.626],[408,1.2369],[410,0.8588],[412,0.6399],[414,0.4881],[416,0.3478],[418,0.2643],[420,0.3041],[422,0.2908],[424,0.253],[426,0.1869],[428,0.1381],[430,0.1189],[432,0.0949],[434,0.0721],[436,0.0661],[438,0.0634],[440,0.0564],[442,0.0605],[444,0.061],[446,0.0715],[448,0.0722],[450,0.0869],[452,0.1117],[454,0.1476],[456,0.2061],[458,0.2691],[460,0.3544],[462,0.5198],[464,0.6961],[466,1.0419],[468,1.0688],[470,0.5714],[472,0.4212],[474,0.1309],[476,-0.0819],[478,0.1424],[480,0.591],[482,1.8695],[484,2.22],[486,2.178],[488,1.5462],[490,0.9896],[492,0.6253],[494,0.5267],[496,0.443],[498,0.3724],[500,0.3408],[502,0.2897],[504,0.3086],[506,0.3926],[508,0.4343],[510,0.5049],[512,0.5428],[514,0.2921],[516,-0.1158],[518,-0.1034],[520,0.9648],[522,2.4554],[524,3.2422],[526,3.5135],[528,3.4974],[530,3.3962],[532,3.3019],[534,1.1005],[536,0.1811],[538,0.4251],[540,0.7801],[542,2.0204],[544,1.7839],[546,1.5081],[548,1.0627],[550,0.8626],[552,0.851],[554,0.8323],[556,0.7841],[558,0.774],[560,0.7592],[562,0.769],[564,0.767],[566,0.7825],[568,0.7836],[570,0.753],[572,0.7819],[574,0.7831],[576,0.7872],[578,0.7949],[580,0.7936],[582,0.7355],[584,0.5701],[586,0.5032],[588,0.5444],[590,0.6164],[592,0.6081],[594,0.5942],[596,0.3388],[598,0.2036],[600,0.0514],[602,0.0418],[604,-0.0224],[606,-0.0298],[608,-0.0471],[610,-0.0491],[612,-0.0505],[618,-0.0501],[622,-0.0499],[624,-0.0486],[626,-0.0489],[628,-0.0521],[630,-0.0599],[632,-0.0639],[634,-0.0703],[636,-0.0656],[638,-0.0659],[640,-0.0624],[642,-0.0614],[644,-0.0671],[646,-0.0651],[648,-0.0624],[650,-0.0593],[652,-0.0569],[654,-0.052],[656,-0.053],[658,-0.0515],[660,-0.0481],[662,-0.0443],[664,-0.0447],[666,-0.0378],[668,-0.0365],[670,-0.0327],[672,-0.0293],[674,-0.0169],[676,-0.0094],[678,-0.0091],[680,-0.0032],[682,-0.0062],[684,-0.006],[686,-0.007],[688,-0.0055],[690,-0.0122],[692,-0.0113],[694,-0.0101],[696,-0.0105],[698,-0.02],[700,-0.018],[702,-0.0127],[704,-0.0059],[706,0.0002],[708,0.007],[710,0.0114],[712,0.0114],[714,0.0168],[716,0.0291],[718,0.0338],[720,0.0338]]
       }
       CTRL_C_eyesAim.Location.Z {
        type=MovieSceneScriptingFloatChannel
        default=0
        thinned_from=492
        keys [[-262,0.739],[-260,0.6464],[-258,0.6105],[-256,0.6068],[-254,0.414],[-252,0.2694],[-250,0.1318],[-248,0.0472],[-246,-0.0281],[-244,-0.0472],[-242,0.0135],[-240,-0.0735],[-238,0.0089],[-236,0.1678],[-234,0.181],[-232,0.2624],[-230,0.329],[-228,0.4698],[-226,0.5558],[-224,0.6053],[-222,0.665],[-220,0.7416],[-218,0.7512],[-216,0.7146],[-214,0.6941],[-212,0.6847],[-210,0.5284],[-208,0.4856],[-206,0.4042],[-204,0.4368],[-202,0.407],[-200,0.3171],[-198,0.2778],[-196,0.2193],[-194,0.1635],[-192,0.1282],[-190,0.1153],[-188,0.0458],[-186,0.0426],[-184,0.0041],[-182,-0.0326],[-180,-0.0811],[-178,-0.1234],[-176,-0.1561],[-174,-0.2364],[-172,-0.3197],[-170,-0.3057],[-168,-0.2132],[-166,-0.1816],[-164,-0.0919],[-162,0.0098],[-160,0.0738],[-158,0.0832],[-156,0.1692],[-154,0.3295],[-152,0.3217],[-150,0.3695],[-148,0.3494],[-146,0.302],[-144,0.2492],[-142,0.2254],[-140,0.0339],[-138,-0.0039],[-136,-0.0065],[-134,-0.0174],[-132,0.0472],[-130,0.1805],[-128,0.1717],[-126,0.2297],[-124,0.2152],[-122,0.2002],[-120,0.2002],[-118,0.2481],[-116,0.2825],[-114,0.272],[-112,0.2367],[-110,0.2073],[-108,0.1828],[-106,0.135],[-104,0.1116],[-102,0.0721],[-100,-0.0808],[-98,-0.0986],[-96,-0.0964],[-94,-0.0337],[-92,-0.0432],[-90,-0.0696],[-88,-0.0769],[-86,-0.0812],[-84,-0.0146],[-82,0.0067],[-80,-0.0272],[-76,0.0057],[-74,0.1471],[-72,0.1597],[-70,0.1878],[-68,0.2981],[-66,0.362],[-64,0.4144],[-62,0.4302],[-60,0.4556],[-58,0.4823],[-56,0.5188],[-54,0.5151],[-52,0.4779],[-50,0.4292],[-48,0.4191],[-46,0.379],[-44,1.172],[-42,1.1886],[-40,0.8484],[-38,0.4469],[-36,0.3594],[-34,0.3701],[-32,0.3362],[-30,0.2661],[-28,0.1717],[-26,0.1874],[-24,0.1794],[-22,0.1368],[-20,0.1425],[-18,0.0596],[-16,0.0655],[-14,0.0955],[-12,0.1248],[-10,-0.0082],[-8,-0.0032],[-6,-0.0118],[-4,0.0065],[-2,0.0792],[0,0.1019],[2,0.1187],[4,0.1377],[6,0.1955],[8,0.2177],[10,0.2423],[12,0.2532],[14,0.2366],[16,0.2309],[18,0.215],[20,0.2004],[22,0.1926],[24,0.1973],[26,0.2256],[28,0.2121],[30,0.1487],[32,0.1141],[34,0.1564],[36,0.1644],[38,0.1626],[40,0.152],[42,0.0927],[44,0.106],[46,0.1135],[48,0.1234],[50,0.0972],[52,0.1264],[54,0.1635],[56,0.1568],[58,0.1386],[60,0.069],[62,0.0385],[64,0.0879],[66,0.0517],[68,-0.0168],[70,-0.0112],[72,0.0333],[74,0.0324],[76,0.0026],[78,0.0013],[80,-0.0311],[82,-0.0452],[84,-0.1047],[86,-0.176],[88,-0.2219],[90,-0.22],[92,-0.3471],[94,-0.3066],[96,-0.3451],[98,-0.3078],[100,-0.3231],[102,-0.3246],[104,-0.432],[106,-0.5373],[108,-0.5586],[110,-0.5179],[112,-0.4091],[114,-0.2286],[116,-0.0559],[118,-0.0865],[120,-0.1067],[122,-0.2057],[124,-0.2227],[126,-0.219],[128,-0.1602],[130,-0.1602],[132,-0.205],[134,-0.1642],[136,-0.1655],[138,-0.1938],[140,-0.1686],[142,-0.1175],[144,-0.102],[146,-0.0382],[148,0.0267],[150,-0.0538],[152,-0.0049],[154,-0.0272],[156,-0.0264],[158,-0.0132],[160,0.0676],[162,-0.0283],[164,0.0632],[166,0.0443],[168,0.001],[170,-0.0178],[172,-0.0384],[174,-0.0466],[176,-0.0639],[178,-0.1476],[180,-0.148],[182,-0.2838],[184,-0.3736],[186,-0.4003],[188,-0.4735],[190,-0.5038],[192,-0.4861],[194,-0.478],[196,-0.4423],[198,-0.4863],[200,-0.5916],[202,-0.6257],[204,-0.6234],[206,-0.6898],[208,-0.7642],[210,-0.7682],[212,-0.8926],[214,-0.9771],[216,-1.1114],[218,-1.2335],[220,-1.2967],[222,-1.4055],[224,-1.4818],[226,-1.6563],[228,-1.6718],[230,-1.8337],[232,-1.9851],[234,-2.1022],[236,-2.2843],[238,-2.42],[240,-2.5717],[242,-2.6442],[244,-2.6477],[246,-2.7555],[248,-2.824],[250,-2.8097],[252,-2.8548],[254,-2.873],[256,-2.8563],[258,-2.8274],[260,-2.8597],[262,-2.8448],[264,-2.8204],[266,-2.9095],[268,-3.1024],[270,-3.2283],[272,-3.287],[274,-3.5531],[276,-3.661],[278,-3.3948],[280,-3.3785],[282,-3.3371],[284,-3.3407],[288,-3.1503],[290,-3.0613],[292,-3.0038],[294,-3.0488],[296,-3.0181],[298,-3.0837],[300,-3.1304],[302,-3.1844],[304,-3.1521],[306,-3.1606],[308,-3.1627],[310,-3.0342],[312,-2.9849],[314,-2.9455],[316,-2.8618],[318,-2.8256],[320,-2.7059],[322,-2.6495],[324,-2.5572],[326,-2.5687],[328,-2.5513],[330,-2.5193],[332,-2.4249],[334,-2.3557],[336,-2.2277],[338,-2.2093],[340,-2.1589],[342,-2.1134],[344,-2.0636],[346,-1.9649],[348,-1.8377],[350,-1.7403],[352,-1.7067],[354,-1.6945],[356,-1.6301],[358,-1.5426],[360,-1.5466],[362,-1.5098],[364,-1.4504],[366,-1.3325],[368,-1.2375],[370,-1.1554],[372,-0.9703],[374,-0.6219],[376,-0.557],[378,-0.1321],[380,0.435],[382,0.9631],[384,-0.1842],[386,-1.8914],[388,-4.2265],[390,-5.9961],[392,-8.9408],[394,-9.9625],[396,-9.3512],[398,-7.9286],[400,-7.5092],[402,-6.8142],[404,-6.4301],[406,-5.1808],[408,-4.2208],[410,-3.2858],[412,-2.6608],[414,-2.1283],[416,-1.5434],[418,-1.1685],[420,-1.0265],[422,-0.761],[424,-0.6073],[426,-0.3919],[428,-0.2313],[430,-0.209],[432,-0.0959],[434,0.0353],[436,0.114],[438,0.1493],[440,0.2045],[442,0.2571],[444,0.3054],[446,0.2355],[448,0.2594],[450,0.1767],[452,0.068],[454,-0.1012],[456,-0.332],[458,-0.4743],[460,-0.6682],[462,-1.0462],[464,-1.3928],[466,-1.9085],[468,-1.6009],[470,0.9203],[472,2.451],[474,2.1784],[476,2.1909],[478,4.4475],[480,8.0148],[482,12.4658],[484,13.2679],[486,13.1883],[488,11.6759],[492,8.212],[494,7.7241],[496,7.1709],[500,5.4858],[502,4.8091],[504,3.9963],[506,3.4623],[508,2.8773],[510,2.4367],[512,2.091],[514,1.6976],[516,1.3786],[518,1.341],[520,0.964],[522,0.7801],[524,0.5637],[526,-0.1405],[528,-0.1316],[530,-0.1232],[532,-0.0371],[534,0.5519],[536,0.3756],[538,0.6668],[540,0.7547],[542,0.3951],[544,0.2769],[546,0.3332],[548,0.443],[550,0.6058],[552,0.8746],[554,1.0934],[556,1.1073],[558,1.1032],[560,1.1038],[562,1.1153],[564,1.1914],[566,1.1408],[568,1.129],[570,1.0842],[572,1.1365],[574,1.1358],[576,1.1199],[578,1.0673],[580,1.0507],[582,1.1823],[584,2.3916],[586,3.1677],[588,3.1318],[590,3.2698],[592,2.8208],[594,2.7145],[596,1.9946],[598,1.5762],[600,1.0607],[602,1.072],[604,1.127],[606,1.1121],[608,1.0366],[612,0.7854],[614,0.7071],[616,0.6608],[618,0.6339],[620,0.6265],[622,0.6115],[624,0.5911],[626,0.5967],[628,0.6486],[630,0.7364],[632,0.8487],[634,0.9523],[636,0.9516],[638,0.9718],[640,0.9568],[642,0.879],[644,0.9742],[646,0.9678],[648,0.887],[650,0.8519],[652,0.8615],[654,0.8417],[656,0.8248],[658,0.8608],[660,0.8839],[662,0.8631],[664,0.8647],[666,0.9121],[668,0.9719],[670,0.9861],[672,0.98],[674,0.8066],[676,0.7733],[678,0.7472],[680,0.7482],[682,0.7775],[684,0.8013],[686,0.7383],[688,0.7701],[690,0.7991],[692,0.8321],[694,0.8098],[696,0.8519],[698,1.1087],[700,1.1377],[702,1.0967],[704,1.0446],[706,1.032],[708,1.0234],[710,1.0074],[712,1.0034],[714,0.9718],[716,0.9112],[718,0.9451],[720,0.9451]]
       }
       CTRL_C_eyesAim.Rotation.X {
        type=MovieSceneScriptingFloatChannel
        default=0
        thinned_from=492
        keys [[-262,1.732],[-260,1.5154],[-258,1.4313],[-256,1.4226],[-254,0.971],[-252,0.6321],[-250,0.3093],[-248,0.1108],[-246,-0.066],[-244,-0.1109],[-242,0.0316],[-240,-0.1726],[-238,0.0209],[-236,0.3938],[-234,0.4248],[-232,0.6157],[-230,0.7718],[-228,1.1018],[-226,1.3034],[-224,1.4193],[-222,1.559],[-220,1.7383],[-218,1.7609],[-216,1.6752],[-214,1.6272],[-212,1.605],[-210,1.2392],[-208,1.1389],[-206,0.9481],[-204,1.0246],[-202,0.9548],[-200,0.7441],[-198,0.6518],[-196,0.5145],[-194,0.3838],[-192,0.3009],[-190,0.2707],[-188,0.1075],[-186,0.1001],[-184,0.0096],[-182,-0.0767],[-180,-0.1903],[-178,-0.2897],[-176,-0.3666],[-174,-0.5553],[-172,-0.7512],[-170,-0.7183],[-168,-0.5009],[-166,-0.4266],[-164,-0.2158],[-162,0.0231],[-160,0.1732],[-158,0.1952],[-156,0.3971],[-154,0.7729],[-152,0.7547],[-150,0.8669],[-148,0.8197],[-146,0.7085],[-144,0.5847],[-142,0.5289],[-140,0.0797],[-138,-0.0093],[-136,-0.0152],[-134,-0.0408],[-132,0.1107],[-130,0.4236],[-128,0.4029],[-126,0.539],[-124,0.5049],[-122,0.4697],[-120,0.4699],[-118,0.5823],[-116,0.6628],[-114,0.6382],[-112,0.5555],[-110,0.4864],[-108,0.4289],[-106,0.3169],[-104,0.2619],[-102,0.1693],[-100,-0.1897],[-98,-0.2316],[-96,-0.2263],[-94,-0.0791],[-92,-0.1015],[-90,-0.1634],[-88,-0.1805],[-86,-0.1907],[-84,-0.0344],[-82,0.0157],[-80,-0.0638],[-78,-0.025],[-76,0.0133],[-74,0.3452],[-72,0.3747],[-70,0.4408],[-68,0.6994],[-66,0.8493],[-64,0.972],[-62,1.009],[-60,1.0685],[-58,1.1312],[-56,1.2167],[-54,1.2081],[-52,1.1208],[-50,1.0068],[-48,0.983],[-46,0.8892],[-44,2.7453],[-42,2.784],[-40,1.9884],[-38,1.0482],[-36,0.8431],[-34,0.8682],[-32,0.7887],[-30,0.6243],[-28,0.4029],[-26,0.4397],[-24,0.421],[-22,0.3211],[-20,0.3344],[-18,0.14],[-16,0.1539],[-14,0.2242],[-12,0.2929],[-10,-0.0193],[-8,-0.0076],[-6,-0.0278],[-4,0.0151],[-2,0.186],[0,0.2392],[2,0.2786],[4,0.3231],[6,0.4587],[8,0.511],[10,0.5686],[12,0.594],[14,0.5552],[16,0.5419],[18,0.5045],[20,0.4703],[22,0.452],[24,0.4629],[26,0.5293],[28,0.4978],[30,0.349],[32,0.2677],[34,0.367],[36,0.3859],[38,0.3815],[40,0.3568],[42,0.2176],[44,0.2488],[46,0.2663],[48,0.2896],[50,0.2281],[52,0.2968],[54,0.3836],[56,0.3681],[58,0.3252],[60,0.162],[62,0.0904],[64,0.2064],[66,0.1214],[68,-0.0394],[70,-0.0262],[72,0.0783],[74,0.0761],[76,0.006],[78,0.0031],[80,-0.073],[82,-0.106],[84,-0.246],[86,-0.4134],[88,-0.5213],[90,-0.5168],[92,-0.8158],[94,-0.7203],[96,-0.8108],[98,-0.7233],[100,-0.7593],[102,-0.7626],[104,-1.0153],[106,-1.2632],[108,-1.3133],[110,-1.2176],[112,-0.9614],[114,-0.5371],[116,-0.1313],[118,-0.203],[120,-0.2506],[122,-0.4831],[124,-0.5232],[126,-0.5145],[128,-0.3764],[130,-0.3764],[132,-0.4816],[134,-0.3856],[136,-0.3888],[138,-0.4552],[140,-0.396],[142,-0.2758],[144,-0.2396],[146,-0.0897],[148,0.0627],[150,-0.1263],[152,-0.0115],[154,-0.064],[156,-0.0621],[158,-0.031],[160,0.1586],[162,-0.0664],[164,0.1484],[166,0.104],[168,0.0023],[170,-0.0419],[172,-0.0903],[174,-0.1095],[176,-0.1501],[178,-0.3466],[180,-0.3475],[182,-0.6669],[184,-0.8779],[186,-0.9408],[188,-1.1129],[190,-1.1844],[192,-1.1426],[194,-1.1236],[196,-1.0396],[198,-1.1431],[200,-1.3912],[202,-1.4715],[204,-1.4661],[206,-1.6224],[208,-1.7979],[210,-1.8073],[212,-2.1007],[214,-2.3002],[216,-2.6173],[218,-2.906],[220,-3.0557],[222,-3.3133],[224,-3.494],[226,-3.9079],[228,-3.9447],[230,-4.3294],[232,-4.6895],[234,-4.9685],[236,-5.4028],[238,-5.7271],[240,-6.0902],[242,-6.2643],[244,-6.2728],[246,-6.5322],[248,-6.6966],[250,-6.6626],[252,-6.771],[254,-6.8147],[256,-6.7745],[258,-6.705],[260,-6.7826],[262,-6.7468],[264,-6.688],[266,-6.9014],[268,-7.3654],[270,-7.6696],[272,-7.8136],[274,-8.4642],[276,-8.7385],[278,-8.0994],[280,-8.0681],[282,-7.9702],[284,-7.9831],[288,-7.5158],[290,-7.2981],[292,-7.1582],[294,-7.2667],[296,-7.1929],[298,-7.3517],[300,-7.4658],[302,-7.5969],[304,-7.5177],[306,-7.5384],[308,-7.5432],[310,-7.2295],[312,-7.1099],[314,-7.0141],[316,-6.8105],[318,-6.7219],[320,-6.4335],[322,-6.2975],[324,-6.0746],[326,-6.1026],[328,-6.0617],[330,-5.9841],[332,-5.7572],[334,-5.5915],[336,-5.2848],[338,-5.2403],[340,-5.119],[344,-4.8882],[346,-4.6505],[348,-4.346],[350,-4.113],[352,-4.0324],[354,-4.0033],[356,-3.8498],[358,-3.6417],[360,-3.651],[362,-3.5635],[364,-3.4225],[366,-3.1428],[368,-2.9176],[370,-2.7231],[372,-2.2851],[374,-1.4629],[376,-1.3097],[378,-0.3104],[380,1.0203],[382,2.2565],[384,-0.4327],[386,-4.4688],[388,-10.133],[390,-14.6054],[392,-22.5718],[394,-25.429],[396,-23.4569],[398,-19.4791],[400,-18.3637],[402,-16.5479],[404,-15.5413],[406,-12.428],[408,-10.0684],[410,-7.801],[412,-6.3],[414,-5.0288],[416,-3.6393],[418,-2.7519],[420,-2.4166],[422,-1.7902],[424,-1.4281],[426,-0.9211],[428,-0.5433],[430,-0.4909],[432,-0.2252],[434,0.0829],[436,0.2676],[438,0.3505],[440,0.48],[442,0.6034],[444,0.7166],[446,0.5526],[448,0.6088],[450,0.4148],[452,0.1596],[454,-0.2376],[456,-0.7801],[458,-1.1149],[460,-1.5715],[462,-2.4633],[464,-3.2832],[466,-4.5075],[468,-3.7769],[470,2.1564],[472,5.7317],[474,5.0957],[476,5.1249],[478,10.3942],[480,18.8182],[482,29.7394],[484,31.7909],[486,31.5845],[488,27.7467],[490,23.453],[492,19.29],[494,18.1234],[496,16.8074],[500,12.8302],[502,11.242],[504,9.3398],[506,8.0922],[508,6.7265],[510,5.6982],[512,4.8914],[514,3.9731],[516,3.2281],[518,3.1402],[520,2.2586],[522,1.828],[524,1.3213],[526,-0.3299],[528,-0.309],[530,-0.2893],[532,-0.0871],[534,1.2941],[536,0.881],[538,1.5632],[540,1.7689],[542,0.9267],[544,0.6496],[546,0.7817],[548,1.0389],[550,1.4203],[552,2.0494],[554,2.5613],[556,2.5938],[558,2.5842],[560,2.5855],[562,2.6125],[564,2.7904],[566,2.6721],[568,2.6444],[570,2.5398],[572,2.662],[574,2.6604],[576,2.6231],[578,2.5001],[580,2.4613],[582,2.769],[584,5.5929],[586,7.4045],[588,7.3207],[590,7.6431],[592,6.5948],[594,6.3466],[596,4.6665],[598,3.6896],[600,2.4851],[602,2.5114],[604,2.6402],[606,2.6051],[608,2.4287],[612,1.8411],[614,1.6576],[616,1.5493],[618,1.4863],[620,1.469],[622,1.4338],[624,1.386],[626,1.3992],[628,1.5207],[630,1.7264],[632,1.9892],[634,2.2314],[636,2.2298],[638,2.277],[640,2.2419],[642,2.0599],[644,2.2828],[646,2.2679],[648,2.0788],[650,1.9965],[652,2.0192],[654,1.9728],[656,1.9332],[658,2.0174],[660,2.0714],[662,2.0227],[664,2.0265],[666,2.1374],[668,2.2773],[670,2.3105],[672,2.2963],[674,1.8906],[676,1.8125],[678,1.7515],[680,1.7539],[682,1.8225],[684,1.8783],[686,1.7308],[688,1.8052],[690,1.8731],[692,1.9502],[694,1.8981],[696,1.9966],[698,2.5973],[700,2.665],[702,2.5691],[704,2.4473],[706,2.4179],[708,2.3979],[710,2.3604],[712,2.3511],[714,2.2771],[716,2.1353],[718,2.2146],[720,2.2146]]
       }
       CTRL_C_eyesAim.Rotation.Y {
        type=MovieSceneScriptingFloatChannel
        default=0
        thinned_from=492
        keys [[-262,-0.5391],[-260,-0.4328],[-258,-0.4156],[-256,-0.4094],[-254,-0.2644],[-252,-0.162],[-250,-0.0771],[-248,-0.0264],[-246,0.0158],[-244,0.0266],[-242,-0.0075],[-240,0.0399],[-238,-0.0048],[-236,-0.091],[-234,-0.098],[-232,-0.1409],[-230,-0.1743],[-228,-0.24],[-226,-0.2727],[-224,-0.2796],[-222,-0.2937],[-220,-0.3251],[-218,-0.3428],[-216,-0.3348],[-214,-0.3288],[-212,-0.3249],[-210,-0.2534],[-208,-0.2234],[-206,-0.1771],[-204,-0.1855],[-202,-0.1656],[-200,-0.1225],[-198,-0.1022],[-196,-0.0758],[-194,-0.0535],[-192,-0.0403],[-190,-0.0358],[-188,-0.0134],[-186,-0.0124],[-182,0.0099],[-180,0.0254],[-178,0.0413],[-176,0.0556],[-174,0.0907],[-172,0.1297],[-170,0.1301],[-168,0.0949],[-166,0.0828],[-164,0.0433],[-162,-0.0046],[-160,-0.0351],[-158,-0.0397],[-156,-0.0829],[-154,-0.1639],[-152,-0.1573],[-150,-0.1852],[-148,-0.1664],[-146,-0.1379],[-144,-0.1098],[-142,-0.0965],[-140,-0.0149],[-138,0.0017],[-136,0.0028],[-134,0.0077],[-132,-0.0209],[-130,-0.0815],[-128,-0.0775],[-126,-0.1068],[-124,-0.0997],[-122,-0.0929],[-120,-0.0931],[-118,-0.1146],[-116,-0.1331],[-114,-0.128],[-112,-0.111],[-110,-0.0976],[-108,-0.0854],[-106,-0.0628],[-104,-0.0522],[-102,-0.0341],[-100,0.0391],[-98,0.0478],[-96,0.0491],[-94,0.018],[-92,0.0233],[-90,0.0381],[-88,0.0423],[-86,0.0454],[-84,0.0084],[-82,-0.0039],[-80,0.0153],[-78,0.0058],[-76,-0.0031],[-74,-0.0772],[-72,-0.0822],[-70,-0.0954],[-68,-0.1524],[-66,-0.1832],[-64,-0.2084],[-62,-0.2171],[-58,-0.2456],[-56,-0.2647],[-54,-0.2641],[-52,-0.2446],[-50,-0.2142],[-48,-0.206],[-46,-0.1121],[-44,-0.2365],[-42,-0.4397],[-40,-0.3542],[-38,-0.187],[-36,-0.1537],[-34,-0.1652],[-32,-0.1561],[-30,-0.1243],[-28,-0.0792],[-26,-0.0865],[-24,-0.0835],[-22,-0.062],[-20,-0.066],[-18,-0.0281],[-16,-0.0306],[-14,-0.0454],[-12,-0.0597],[-10,0.0042],[-8,0.0016],[-6,0.0061],[-4,-0.0034],[-2,-0.0413],[0,-0.0518],[2,-0.0605],[4,-0.0689],[6,-0.0988],[8,-0.1101],[10,-0.1238],[12,-0.131],[14,-0.1236],[16,-0.1205],[18,-0.1122],[20,-0.1062],[22,-0.1031],[24,-0.1061],[26,-0.1216],[28,-0.1161],[30,-0.0817],[32,-0.0625],[34,-0.0858],[36,-0.0914],[38,-0.0908],[40,-0.0858],[42,-0.0515],[44,-0.0598],[46,-0.0642],[48,-0.0701],[50,-0.0551],[52,-0.0727],[54,-0.0944],[56,-0.0912],[58,-0.0813],[60,-0.0412],[62,-0.0236],[64,-0.0545],[66,-0.0322],[68,0.0105],[70,0.007],[72,-0.0214],[74,-0.0204],[76,-0.0016],[78,-0.0008],[80,0.0196],[82,0.0286],[84,0.0656],[86,0.1113],[88,0.1415],[90,0.14],[92,0.2182],[94,0.189],[96,0.2029],[98,0.1746],[100,0.1733],[102,0.1539],[104,0.1875],[106,0.1786],[108,0.1789],[110,0.1437],[112,0.0957],[114,0.0481],[116,0.0116],[118,0.0182],[120,0.0226],[122,0.0426],[124,0.0455],[126,0.0444],[128,0.0291],[130,0.0291],[132,0.0339],[134,0.0268],[136,0.028],[138,0.0329],[140,0.0321],[142,0.024],[144,0.0225],[146,0.0102],[148,-0.0078],[150,0.0167],[152,0.0017],[154,0.0101],[156,0.0099],[158,0.0051],[160,-0.0277],[162,0.012],[164,-0.027],[166,-0.0193],[168,-0.0004],[170,0.0081],[172,0.0176],[174,0.0215],[176,0.0304],[178,0.0712],[180,0.0703],[182,0.1354],[184,0.1739],[186,0.1862],[188,0.2204],[190,0.2345],[192,0.2207],[194,0.2162],[196,0.197],[198,0.2126],[200,0.2605],[202,0.28],[204,0.2804],[206,0.3127],[208,0.355],[210,0.3574],[212,0.4179],[214,0.463],[216,0.5374],[218,0.606],[220,0.6595],[222,0.7156],[224,0.7625],[226,0.862],[228,0.8779],[230,1.0103],[232,1.104],[234,1.1761],[236,1.2956],[238,1.4053],[240,1.5185],[242,1.6282],[244,1.6786],[246,1.8466],[248,1.8908],[250,1.9172],[252,1.9552],[254,1.9595],[256,1.9518],[258,1.9172],[260,1.9343],[262,1.9367],[264,1.8772],[266,1.8569],[268,1.9634],[270,2.1035],[272,2.384],[274,3.0571],[276,3.8723],[278,4.1377],[280,4.5518],[282,4.6534],[284,4.8563],[286,4.6788],[288,4.4878],[290,4.3076],[292,4.2414],[296,4.2439],[298,4.2961],[300,4.3831],[302,4.4318],[304,4.3647],[308,4.3584],[310,4.1289],[312,4.0763],[314,4.0062],[316,3.8696],[318,3.7559],[320,3.673],[322,3.6243],[324,3.4954],[326,3.525],[328,3.574],[330,3.5185],[332,3.4215],[334,3.407],[336,3.3202],[338,3.2808],[340,3.1873],[342,3.0044],[344,2.8292],[348,2.284],[350,2.0703],[352,1.9685],[354,1.9349],[356,1.8084],[358,1.7043],[360,1.6885],[362,1.6395],[364,1.5728],[366,1.4404],[368,1.3265],[370,1.219],[372,0.9969],[374,0.594],[376,0.5249],[378,0.1176],[380,-0.3531],[382,-0.7298],[384,0.1399],[386,1.5544],[388,4.7608],[390,7.1659],[392,10.9295],[394,11.2759],[396,8.5496],[398,5.6413],[400,5.008],[402,3.9417],[404,2.3971],[406,2.3801],[408,1.7594],[410,1.1306],[412,0.7842],[414,0.5618],[416,0.3667],[418,0.2543],[420,0.2829],[422,0.226],[424,0.1727],[426,0.0989],[428,0.052],[430,0.0433],[432,0.0187],[434,-0.0066],[436,-0.0215],[438,-0.0284],[440,-0.0386],[442,-0.0514],[444,-0.0629],[446,-0.0494],[448,-0.0553],[450,-0.0388],[452,-0.0158],[454,0.0252],[456,0.0915],[458,0.1474],[460,0.2352],[462,0.4379],[464,0.6709],[466,1.1359],[468,1.0081],[470,-0.5118],[472,-1.2669],[474,-0.7777],[476,-0.3782],[478,-1.4556],[480,-2.195],[482,-1.0679],[484,-0.6561],[486,-0.3538],[488,-1.0345],[490,-1.6692],[492,-2.1803],[494,-2.1054],[496,-2.0974],[498,-2.1666],[500,-2.151],[502,-1.9341],[504,-1.775],[506,-1.7242],[508,-1.5052],[510,-1.3517],[512,-1.189],[514,-0.7517],[516,0.0115],[518,0.0918],[520,0.6874],[522,0.906],[524,0.7683],[526,-0.1986],[528,-0.1855],[530,-0.1705],[532,-0.0505],[534,0.4137],[536,0.1193],[538,-0.3188],[540,-0.4812],[542,-0.4076],[544,-0.2657],[546,-0.2923],[548,-0.3246],[550,-0.4027],[552,-0.5847],[554,-0.7297],[556,-0.7189],[560,-0.7057],[562,-0.7177],[564,-0.768],[566,-0.7408],[568,-0.7333],[570,-0.69],[572,-0.7377],[574,-0.7377],[576,-0.7287],[578,-0.6963],[580,-0.6844],[582,-0.7472],[584,-1.3934],[586,-1.7495],[588,-1.7866],[590,-1.9599],[592,-1.6895],[594,-1.6107],[596,-0.9426],[598,-0.614],[600,-0.275],[602,-0.2694],[604,-0.2123],[606,-0.1979],[608,-0.1518],[610,-0.1169],[612,-0.0839],[614,-0.0655],[616,-0.0546],[618,-0.0486],[620,-0.0468],[622,-0.0434],[624,-0.0412],[626,-0.0418],[628,-0.0483],[630,-0.0537],[632,-0.0738],[634,-0.0861],[636,-0.0965],[638,-0.1012],[640,-0.1045],[642,-0.0869],[644,-0.0993],[646,-0.1019],[648,-0.0868],[650,-0.0841],[652,-0.091],[654,-0.095],[656,-0.0891],[660,-0.1116],[662,-0.1123],[664,-0.1122],[666,-0.134],[668,-0.1511],[670,-0.1601],[672,-0.1631],[674,-0.1335],[676,-0.133],[678,-0.1267],[680,-0.1327],[682,-0.1373],[684,-0.1437],[686,-0.1266],[688,-0.1362],[690,-0.1367],[692,-0.1462],[694,-0.1416],[696,-0.1522],[698,-0.2104],[700,-0.2211],[702,-0.2165],[704,-0.2101],[706,-0.2136],[708,-0.2187],[710,-0.2188],[712,-0.2176],[714,-0.214],[716,-0.2079],[718,-0.2224],[720,-0.2224]]
       }
       CTRL_C_eyesAim.Rotation.Z {
        type=MovieSceneScriptingFloatChannel
        default=0
        thinned_from=492
        keys [[-262,17.2845],[-260,15.9368],[-258,16.1869],[-256,16.0519],[-254,15.2296],[-252,14.3711],[-250,13.9974],[-248,13.4205],[-244,13.495],[-242,13.2841],[-240,13.0161],[-234,12.9935],[-232,12.8914],[-230,12.7265],[-228,12.2878],[-226,11.8149],[-224,11.1423],[-222,10.6657],[-220,10.5907],[-218,11.0121],[-216,11.2981],[-214,11.4196],[-212,11.4401],[-210,11.5549],[-208,11.0953],[-206,10.578],[-204,10.259],[-202,9.8359],[-200,9.3482],[-198,8.9096],[-196,8.3836],[-194,7.943],[-192,7.6333],[-190,7.5268],[-188,7.1068],[-186,7.0819],[-184,7.1057],[-182,7.3871],[-180,7.5914],[-176,8.6172],[-174,9.272],[-172,9.7992],[-168,10.7319],[-166,10.9896],[-164,11.3374],[-160,11.4591],[-158,11.4929],[-156,11.7877],[-154,11.9721],[-152,11.7748],[-150,12.061],[-148,11.4704],[-146,11.012],[-144,10.6361],[-142,10.3402],[-140,10.5859],[-138,10.4813],[-136,10.5202],[-134,10.6476],[-132,10.7092],[-130,10.8841],[-128,10.8843],[-126,11.2038],[-124,11.1648],[-120,11.2071],[-118,11.1311],[-116,11.3504],[-114,11.3429],[-112,11.3035],[-110,11.3499],[-108,11.2569],[-106,11.2119],[-104,11.2753],[-102,11.3803],[-100,11.6434],[-98,11.6712],[-94,12.8213],[-92,12.9064],[-90,13.1334],[-88,13.1814],[-86,13.3818],[-84,13.6824],[-82,13.7772],[-80,13.506],[-78,13.086],[-76,12.9528],[-74,12.6145],[-72,12.3694],[-70,12.2088],[-68,12.2944],[-66,12.1698],[-64,12.1016],[-62,12.1408],[-60,12.2211],[-56,12.2735],[-54,12.3319],[-52,12.3121],[-50,12.0091],[-48,11.8324],[-46,7.1868],[-44,4.9192],[-42,8.9672],[-40,10.0971],[-38,10.1112],[-36,10.328],[-34,10.7744],[-32,11.1976],[-30,11.2573],[-28,11.1157],[-26,11.1266],[-24,11.2184],[-22,10.9336],[-20,11.156],[-18,11.3497],[-16,11.2497],[-14,11.452],[-12,11.5154],[-10,12.1589],[-8,12.2495],[-6,12.4348],[-4,12.5289],[-2,12.5224],[0,12.226],[2,12.2574],[4,12.0466],[6,12.1588],[8,12.154],[12,12.4348],[14,12.5521],[18,12.5383],[20,12.7191],[22,12.8476],[26,12.9413],[28,13.1325],[30,13.1808],[32,13.149],[34,13.1593],[36,13.3276],[38,13.3815],[40,13.5136],[42,13.3076],[44,13.5184],[48,13.6123],[50,13.5728],[52,13.761],[54,13.8245],[58,14.0283],[60,14.2679],[62,14.6454],[64,14.7913],[66,14.835],[68,14.9216],[70,14.9694],[72,15.2813],[74,15.035],[76,14.8365],[78,14.8096],[80,15.0118],[82,15.0791],[84,14.9272],[86,15.0702],[88,15.1822],[90,15.158],[92,14.9741],[94,14.6998],[96,14.0488],[98,13.5689],[100,12.8592],[102,11.4105],[104,10.462],[106,8.0449],[108,7.7542],[110,6.7286],[112,5.6819],[114,5.1192],[116,5.0586],[118,5.1362],[120,5.1537],[122,5.0417],[124,4.9651],[126,4.9329],[128,4.4205],[130,4.4205],[132,4.0212],[134,3.9722],[136,4.1164],[138,4.1329],[140,4.6373],[142,4.9698],[144,5.3621],[146,6.4867],[148,7.1248],[150,7.5423],[152,8.2552],[154,9.0019],[156,9.0259],[158,9.4115],[160,9.9149],[162,10.232],[164,10.3097],[166,10.5187],[168,10.7858],[170,11.0056],[172,11.0124],[174,11.1101],[176,11.4314],[178,11.6108],[180,11.4347],[182,11.4744],[184,11.2065],[190,11.1994],[192,10.931],[194,10.8908],[196,10.7304],[198,10.5341],[200,10.6058],[202,10.7714],[204,10.8242],[206,10.9075],[208,11.1661],[210,11.1811],[212,11.2451],[214,11.3752],[216,11.5947],[218,11.7688],[220,12.1681],[222,12.1746],[228,12.5274],[230,13.1106],[232,13.2184],[234,13.2845],[236,13.4456],[238,13.7416],[240,13.948],[242,14.5126],[244,14.9218],[246,15.7171],[248,15.6958],[250,15.9818],[252,16.0323],[254,15.9675],[256,15.9984],[258,15.8847],[260,15.8435],[262,15.9434],[264,15.608],[266,14.9869],[268,14.8446],[270,15.2462],[272,16.8637],[274,19.7166],[276,23.7186],[278,26.8858],[280,29.2423],[282,30.0897],[284,31.1173],[286,30.9578],[290,30.3913],[292,30.4938],[294,30.1445],[296,30.386],[298,30.1402],[300,30.2506],[304,29.9715],[308,29.8515],[310,29.5786],[312,29.6787],[316,29.4693],[318,29.0652],[320,29.6023],[322,29.8046],[324,29.8081],[326,29.9024],[328,30.4141],[330,30.3475],[332,30.6234],[334,31.2585],[336,32.0519],[340,31.8262],[344,29.9911],[346,28.7436],[348,27.6725],[350,26.6735],[352,25.978],[354,25.7543],[356,25.1246],[358,25.0469],[360,24.7862],[362,24.6757],[366,24.5989],[368,24.4282],[370,24.0981],[372,23.5585],[374,22.0951],[376,21.8337],[378,20.7531],[380,19.0893],[382,17.914],[384,17.9122],[386,19.1415],[388,24.9098],[390,25.5809],[392,24.5186],[394,22.3551],[396,18.9119],[398,15.5312],[400,14.7335],[402,13.0266],[404,8.5527],[406,10.6718],[408,9.8104],[410,8.1956],[412,7.0673],[414,6.358],[416,5.7461],[418,5.2756],[420,6.6732],[422,7.1938],[424,6.8921],[426,6.1252],[428,5.463],[430,5.0401],[432,4.751],[434,4.5242],[436,4.6031],[438,4.6341],[440,4.6017],[442,4.8724],[444,5.0173],[448,5.1928],[450,5.3472],[452,5.6687],[454,6.0508],[456,6.6898],[458,7.5292],[460,8.5107],[462,10.0744],[464,11.5365],[466,14.1158],[468,14.9229],[470,13.345],[472,12.4223],[474,8.6549],[476,4.2093],[478,7.8847],[480,6.4124],[482,1.8685],[484,1.0583],[486,0.5753],[488,1.9655],[490,3.8412],[492,6.2035],[494,6.4043],[496,6.9084],[500,9.3581],[502,9.6366],[504,10.6654],[506,11.9484],[508,12.5561],[510,13.3011],[512,13.6298],[514,10.696],[516,-0.2031],[518,-1.6738],[520,-16.9204],[522,-26.3552],[524,-30.1715],[526,-31.0399],[528,-30.9721],[530,-30.5134],[532,-30.129],[534,-17.725],[536,-7.7119],[538,11.5241],[540,15.2125],[542,23.7391],[544,22.2411],[546,20.4994],[548,17.3498],[550,15.8258],[552,15.9178],[554,15.8914],[556,15.4801],[558,15.3882],[560,15.2569],[562,15.3509],[564,15.3765],[566,15.4849],[568,15.4883],[570,15.1894],[572,15.478],[578,15.5519],[580,15.5298],[582,15.0893],[584,13.9456],[586,13.2202],[588,13.6409],[590,14.2975],[592,14.3065],[594,14.1827],[596,11.3947],[598,9.4354],[600,6.3102],[602,6.1195],[604,4.5947],[606,4.3406],[608,3.575],[610,3.1304],[612,2.6069],[614,2.2634],[616,2.0184],[618,1.8711],[620,1.8235],[622,1.7333],[624,1.7011],[626,1.7127],[628,1.8172],[630,1.7798],[632,2.1226],[634,2.2098],[636,2.4781],[638,2.5431],[640,2.6684],[642,2.4143],[644,2.4901],[646,2.572],[648,2.3887],[650,2.4121],[652,2.5807],[654,2.7547],[656,2.6378],[658,2.8488],[660,3.0822],[662,3.1768],[664,3.1666],[666,3.5863],[668,3.7941],[670,3.9621],[672,4.0613],[674,4.037],[676,4.1964],[678,4.1362],[680,4.3268],[682,4.3077],[684,4.3742],[686,4.1828],[688,4.3126],[690,4.1732],[692,4.2847],[694,4.2658],[696,4.3563],[698,4.6284],[700,4.7399],[702,4.8127],[704,4.9028],[706,5.0444],[708,5.2075],[710,5.2932],[712,5.2847],[714,5.3654],[716,5.5589],[718,5.732],[720,5.7321]]
       }
       CTRL_C_eyesAim.Scale.X {type=MovieSceneScriptingFloatChannel, default=1, const [1,492]}
       CTRL_C_eyesAim.Scale.Y {type=MovieSceneScriptingFloatChannel, default=1, const [1,492]}
       CTRL_C_eyesAim.Scale.Z {type=MovieSceneScriptingFloatChannel, default=1, const [1,492]}
       CTRL_L_eyeAim.Location.X {type=MovieSceneScriptingFloatChannel, default=0, const [0,492]}
       CTRL_L_eyeAim.Location.Y {type=MovieSceneScriptingFloatChannel, default=0, const [0,492]}
       CTRL_L_eyeAim.Location.Z {type=MovieSceneScriptingFloatChannel, default=0, const [0,492]}
       CTRL_L_eyeAim.Rotation.X {type=MovieSceneScriptingFloatChannel, default=0, const [0,492]}
       CTRL_L_eyeAim.Rotation.Y {type=MovieSceneScriptingFloatChannel, default=0, const [0,492]}
       CTRL_L_eyeAim.Rotation.Z {type=MovieSceneScriptingFloatChannel, default=0, const [0,492]}
       CTRL_L_eyeAim.Scale.X {type=MovieSceneScriptingFloatChannel, default=1, const [1,492]}
       CTRL_L_eyeAim.Scale.Y {type=MovieSceneScriptingFloatChannel, default=1, const [1,492]}
       CTRL_L_eyeAim.Scale.Z {type=MovieSceneScriptingFloatChannel, default=1, const [1,492]}
       CTRL_R_eyeAim.Location.X {
        type=MovieSceneScriptingFloatChannel
        default=0
        thinned_from=492
        keys [[-262,0.0308],[-260,0.0136],[-258,0.0094],[-256,0.0084],[-254,-0.0011],[-252,0.0022],[-250,-0.0017],[-248,-0.0111],[-246,-0.0162],[-244,-0.0174],[-242,-0.0155],[-240,-0.0429],[-238,-0.0441],[-236,-0.0178],[-234,-0.0175],[-232,-0.0173],[-230,-0.0115],[-228,-0.0006],[-226,-0.0028],[-224,0.0005],[-222,0.0092],[-220,0.0147],[-218,0.0097],[-216,0.0192],[-214,0.0087],[-212,0.0064],[-210,0.0054],[-208,-0.0067],[-206,-0.0138],[-204,-0.0136],[-202,-0.016],[-200,-0.0203],[-198,-0.0146],[-196,-0.0225],[-194,-0.0191],[-192,-0.0316],[-190,-0.0365],[-188,-0.0275],[-186,-0.0269],[-184,-0.0215],[-182,-0.0384],[-180,-0.0333],[-178,-0.0287],[-176,-0.0396],[-174,-0.0481],[-172,-0.0527],[-170,-0.0518],[-168,-0.0512],[-166,-0.0545],[-164,-0.0527],[-162,-0.0588],[-160,-0.0722],[-158,-0.0722],[-156,-0.0711],[-154,-0.0362],[-152,-0.0154],[-150,-0.0025],[-148,0.01],[-146,0.0121],[-144,0.0116],[-142,0.0205],[-140,-0.0056],[-138,0.0211],[-136,0.009],[-134,0.014],[-132,0.0186],[-130,0.0148],[-128,0.0195],[-126,0.0169],[-124,0.0233],[-122,0.0275],[-120,0.0323],[-118,0.0261],[-116,0.0438],[-114,0.0441],[-112,0.0441],[-110,0.0382],[-108,0.043],[-106,0.0516],[-104,0.0525],[-102,0.0312],[-100,0.0266],[-98,0.0265],[-96,0.0144],[-94,0.0164],[-92,0.0133],[-90,0.0098],[-88,0.0068],[-86,-0.004],[-84,-0.0169],[-82,-0.0268],[-80,-0.0442],[-78,-0.0522],[-76,-0.0503],[-74,-0.0328],[-72,-0.0272],[-70,-0.001],[-68,0.0013],[-66,0.0138],[-64,0.0195],[-62,0.0256],[-60,0.0416],[-58,0.0425],[-56,0.0433],[-54,0.0431],[-52,0.0411],[-50,0.0464],[-48,0.0489],[-46,0.033],[-44,-0.0054],[-42,0.0004],[-40,0.0093],[-38,-0.0001],[-36,-0.0009],[-34,0.011],[-32,0.0142],[-30,0.012],[-28,-0.0038],[-26,-0.0061],[-24,-0.0063],[-22,0.0121],[-20,-0.0001],[-18,-0.0102],[-16,-0.0006],[-14,0.0136],[-12,0.0099],[-10,0.0007],[-8,-0.0098],[-6,0.0088],[-4,0.0038],[-2,0.0053],[0,0.0057],[2,-0.0027],[4,0.0028],[6,0.0078],[8,-0.0005],[10,0.0175],[12,-0.0013],[14,-0.0035],[16,-0.0073],[18,-0.0173],[20,-0.0169],[22,-0.0096],[24,-0.0072],[26,-0.0094],[28,-0.0182],[30,-0.0166],[32,-0.0197],[34,-0.0131],[36,-0.0181],[38,-0.0114],[40,-0.0092],[42,-0.0114],[44,-0.0104],[46,-0.013],[48,-0.0216],[50,-0.0262],[52,-0.0253],[54,-0.0183],[56,-0.0163],[58,-0.0151],[60,-0.0218],[62,-0.0193],[64,-0.0124],[66,-0.0236],[68,-0.0206],[70,-0.02],[72,-0.0167],[74,-0.0221],[76,-0.023],[78,-0.0268],[80,-0.0246],[82,-0.0321],[84,-0.0334],[86,-0.0304],[88,-0.0306],[90,-0.0301],[92,-0.0367],[94,-0.0237],[96,-0.0301],[98,-0.0202],[100,-0.007],[102,-0.0134],[104,-0.0264],[106,-0.0219],[108,-0.02],[110,-0.0141],[112,-0.016],[114,-0.0106],[116,-0.0106],[118,-0.0159],[120,-0.0169],[122,-0.008],[126,-0.009],[128,-0.0176],[130,-0.0176],[132,-0.0187],[134,-0.0076],[136,-0.0036],[138,-0.0029],[140,-0.0024],[142,-0.0015],[144,-0.0124],[146,0.0001],[148,-0.0049],[150,-0.0142],[152,0.0077],[154,0.0142],[156,0.013],[158,-0.0072],[160,-0.0097],[162,-0.0154],[164,-0.0016],[166,0.0013],[168,0.0007],[170,0.0007],[172,0.0109],[174,0.0144],[176,0.0148],[178,-0.0074],[180,0.0038],[182,0.0022],[184,-0.0022],[186,-0.0028],[188,-0.0072],[190,-0.0169],[192,-0.0187],[194,-0.0166],[196,-0.0107],[198,-0.017],[200,-0.0228],[202,-0.0178],[204,-0.0145],[206,-0.0217],[208,-0.0203],[210,-0.0201],[212,-0.0258],[214,-0.0374],[216,-0.048],[218,-0.0454],[220,-0.0582],[222,-0.0553],[224,-0.0679],[226,-0.0687],[228,-0.0676],[230,-0.0792],[232,-0.0879],[234,-0.0951],[236,-0.1173],[238,-0.1314],[240,-0.1435],[242,-0.1543],[244,-0.1594],[246,-0.1758],[248,-0.1799],[250,-0.1726],[252,-0.172],[254,-0.1934],[256,-0.1768],[258,-0.1683],[260,-0.1572],[262,-0.1727],[264,-0.1645],[266,-0.1495],[268,-0.1672],[270,-0.1938],[272,-0.2243],[274,-0.2788],[276,-0.3307],[278,-0.3722],[280,-0.4273],[282,-0.4344],[284,-0.4594],[286,-0.4452],[288,-0.4283],[290,-0.3904],[292,-0.4107],[294,-0.4074],[296,-0.3939],[298,-0.3954],[300,-0.4025],[302,-0.4181],[304,-0.4152],[306,-0.4161],[308,-0.4158],[310,-0.3829],[312,-0.3741],[314,-0.3568],[316,-0.3385],[318,-0.3262],[320,-0.3294],[322,-0.328],[324,-0.3169],[326,-0.3198],[328,-0.3354],[330,-0.3245],[332,-0.3118],[334,-0.3095],[336,-0.3004],[338,-0.2889],[340,-0.285],[342,-0.2699],[344,-0.2619],[346,-0.2417],[348,-0.2073],[350,-0.1934],[352,-0.1749],[354,-0.1738],[356,-0.1695],[358,-0.1618],[360,-0.1608],[362,-0.1694],[364,-0.1599],[368,-0.1185],[370,-0.1104],[372,-0.093],[374,-0.0556],[376,-0.0491],[378,-0.0105],[380,0.0334],[382,0.06],[384,-0.0394],[386,-0.1893],[388,-0.5139],[390,-0.746],[392,-1.1072],[394,-1.1403],[396,-0.8667],[398,-0.585],[400,-0.526],[402,-0.4151],[404,-0.274],[406,-0.2645],[408,-0.191],[410,-0.1504],[412,-0.1165],[414,-0.0793],[416,-0.0656],[418,-0.044],[420,-0.0334],[422,-0.0351],[424,-0.031],[426,-0.0259],[428,-0.0322],[430,-0.0163],[432,-0.0149],[434,-0.0147],[436,-0.0137],[438,-0.002],[440,0.0021],[442,-0.0067],[444,0.0026],[446,0.0038],[448,0.0155],[450,0.0184],[452,0.0062],[454,-0.0005],[456,-0.0188],[458,-0.0171],[460,-0.0233],[462,-0.0419],[464,-0.0747],[466,-0.1235],[468,-0.1118],[470,0.0213],[472,0.0612],[474,0.0066],[476,-0.0386],[478,0.057],[480,0.1397],[482,0.0514],[484,0.0162],[486,-0.0157],[488,0.0494],[490,0.1107],[492,0.1677],[494,0.1457],[496,0.1491],[498,0.1578],[500,0.1575],[502,0.1473],[504,0.1242],[506,0.1176],[508,0.085],[510,0.0775],[512,0.0744],[514,0.0245],[516,-0.0495],[518,-0.0592],[520,-0.1418],[522,-0.1899],[524,-0.1938],[526,-0.1027],[528,-0.1073],[530,-0.1096],[532,-0.1124],[534,-0.1454],[536,-0.1045],[538,-0.0243],[540,-0.001],[542,-0.0077],[544,0.0006],[548,-0.0028],[550,-0.0015],[552,0.0213],[554,0.0374],[556,0.0371],[558,0.0422],[560,0.0421],[562,0.031],[564,0.0373],[566,0.0379],[568,0.0375],[570,0.034],[572,0.0482],[574,0.0456],[578,0.0357],[580,0.0344],[582,0.0399],[584,0.0845],[586,0.0925],[588,0.0984],[590,0.1418],[592,0.1175],[594,0.11],[596,0.0603],[598,0.0366],[600,0.0061],[602,0.0067],[604,0.0154],[606,0.0187],[608,0.0275],[610,0.0271],[612,0.0249],[614,0.0234],[616,0.0297],[618,0.0274],[620,0.0246],[622,0.0076],[624,0.006],[626,0.0049],[628,0.0067],[630,0.0276],[632,0.0172],[634,0.0227],[636,0.0145],[638,0.0119],[640,0.0112],[642,0.0092],[644,0.0205],[646,0.0137],[648,0.0099],[650,0.0089],[654,0.0097],[656,0.009],[658,0.0092],[660,-0.0102],[662,0.003],[664,0.0087],[666,0.0021],[668,0.0113],[670,0.0143],[672,0.0158],[674,0.0133],[676,0.0254],[678,0.0173],[680,0.0147],[682,0.0144],[684,0.003],[686,0.0083],[688,0.0125],[692,0.014],[694,0.0032],[696,0.0114],[698,0.0104],[700,0.017],[702,0.0202],[704,0.0205],[706,0.0078],[708,0.0062],[710,0.0055],[712,0.0053],[714,0.0169],[716,0.0193],[718,0.0103],[720,0.0103]]
       }
       CTRL_R_eyeAim.Location.Y {
        type=MovieSceneScriptingFloatChannel
        default=0
        thinned_from=492
        keys [[-262,0.1735],[-260,0.1527],[-258,0.1444],[-256,0.1438],[-254,0.1001],[-252,0.067],[-250,0.0358],[-248,0.0169],[-246,0],[-244,-0.0044],[-242,0.0089],[-240,-0.0109],[-238,0.0077],[-236,0.0427],[-234,0.0457],[-232,0.0642],[-230,0.079],[-228,0.111],[-226,0.131],[-224,0.1426],[-222,0.1564],[-220,0.1739],[-218,0.1763],[-216,0.1679],[-214,0.1633],[-212,0.1612],[-210,0.1259],[-208,0.1171],[-206,0.0984],[-204,0.1066],[-202,0.0988],[-200,0.079],[-198,0.0701],[-196,0.057],[-194,0.0448],[-192,0.0374],[-190,0.0347],[-188,0.0187],[-186,0.018],[-184,0.009],[-182,0.0006],[-180,-0.0105],[-178,-0.0204],[-176,-0.0278],[-174,-0.0467],[-172,-0.0662],[-170,-0.0633],[-168,-0.0426],[-166,-0.0357],[-164,-0.0157],[-162,0.0075],[-160,0.022],[-158,0.024],[-156,0.0425],[-154,0.0789],[-152,0.0775],[-150,0.0873],[-148,0.0834],[-146,0.0735],[-144,0.0612],[-142,0.0562],[-140,0.0125],[-138,0.0041],[-136,0.0035],[-134,0.0007],[-132,0.0155],[-130,0.0455],[-128,0.0434],[-126,0.056],[-122,0.05],[-120,0.0501],[-118,0.0607],[-116,0.0688],[-114,0.0663],[-112,0.0583],[-110,0.0516],[-108,0.0458],[-106,0.0352],[-104,0.0301],[-102,0.0216],[-100,-0.0129],[-98,-0.017],[-96,-0.0172],[-94,-0.003],[-92,-0.0052],[-90,-0.0112],[-88,-0.0129],[-86,-0.014],[-84,0.001],[-82,0.0063],[-80,-0.0009],[-76,0.0065],[-74,0.0381],[-72,0.0414],[-70,0.0475],[-68,0.0725],[-66,0.0866],[-64,0.0983],[-62,0.1018],[-60,0.1069],[-58,0.1126],[-56,0.1203],[-54,0.1201],[-52,0.1117],[-50,0.1003],[-48,0.098],[-46,0.0886],[-44,0.2708],[-42,0.2725],[-40,0.1963],[-38,0.1085],[-36,0.0879],[-34,0.0886],[-32,0.0798],[-30,0.0632],[-28,0.0416],[-26,0.0452],[-24,0.043],[-22,0.0334],[-20,0.0347],[-18,0.0158],[-16,0.0173],[-12,0.0301],[-10,0.0003],[-8,0.0012],[-6,-0.0007],[-4,0.0032],[-2,0.0197],[0,0.0247],[2,0.0285],[4,0.0331],[6,0.0463],[10,0.0566],[12,0.059],[14,0.0553],[16,0.054],[18,0.0504],[20,0.0472],[22,0.0455],[24,0.0466],[26,0.0529],[28,0.0497],[30,0.0357],[32,0.0281],[34,0.0376],[36,0.0396],[38,0.0393],[40,0.0371],[42,0.0241],[44,0.0265],[46,0.0281],[48,0.0302],[50,0.0243],[52,0.031],[54,0.0396],[56,0.0384],[58,0.0342],[60,0.0185],[62,0.0114],[64,0.0222],[66,0.0138],[68,-0.0014],[70,-0.0002],[72,0.0096],[74,0.0097],[76,0.003],[78,0.0029],[80,-0.0044],[82,-0.0079],[84,-0.0215],[86,-0.0376],[88,-0.048],[90,-0.0475],[92,-0.0762],[94,-0.0668],[96,-0.0754],[98,-0.0665],[100,-0.0694],[102,-0.0693],[104,-0.0926],[106,-0.1153],[108,-0.1198],[110,-0.1097],[112,-0.0851],[114,-0.0438],[116,-0.0046],[118,-0.0113],[120,-0.0158],[122,-0.0385],[124,-0.0417],[126,-0.0408],[128,-0.027],[130,-0.027],[132,-0.0363],[134,-0.027],[136,-0.0273],[138,-0.033],[140,-0.028],[142,-0.0163],[144,-0.0134],[146,0.0008],[148,0.0148],[150,-0.0037],[152,0.0063],[154,0.0011],[156,0.0013],[158,0.0045],[160,0.0223],[162,0.0015],[164,0.0211],[166,0.0169],[168,0.0072],[172,-0.0016],[174,-0.0036],[176,-0.007],[178,-0.0264],[180,-0.0263],[182,-0.0572],[184,-0.077],[186,-0.0832],[188,-0.1002],[190,-0.1072],[192,-0.1032],[194,-0.1012],[196,-0.0934],[198,-0.1031],[200,-0.1264],[202,-0.1343],[204,-0.1339],[206,-0.1488],[208,-0.1653],[210,-0.1662],[212,-0.1948],[214,-0.2135],[216,-0.2443],[218,-0.2728],[220,-0.2868],[222,-0.3118],[224,-0.3295],[226,-0.3692],[228,-0.3722],[230,-0.4087],[232,-0.443],[234,-0.4694],[236,-0.511],[238,-0.5423],[240,-0.5767],[242,-0.5929],[244,-0.5944],[246,-0.6202],[248,-0.6358],[250,-0.6328],[252,-0.6428],[254,-0.6467],[256,-0.643],[258,-0.636],[260,-0.6437],[262,-0.6403],[264,-0.6348],[266,-0.6549],[268,-0.6985],[270,-0.7273],[272,-0.7403],[274,-0.8023],[276,-0.8277],[278,-0.7692],[280,-0.7667],[282,-0.7578],[284,-0.7589],[288,-0.7146],[290,-0.6939],[292,-0.6806],[294,-0.691],[296,-0.6843],[298,-0.699],[300,-0.7097],[302,-0.7221],[304,-0.7147],[308,-0.7166],[310,-0.687],[312,-0.6757],[314,-0.6667],[316,-0.6477],[318,-0.6396],[320,-0.6124],[322,-0.5996],[324,-0.5787],[326,-0.581],[328,-0.5775],[330,-0.5701],[332,-0.5488],[334,-0.533],[336,-0.5039],[338,-0.4993],[340,-0.4878],[342,-0.4776],[344,-0.4662],[346,-0.4434],[348,-0.4143],[350,-0.3921],[352,-0.3841],[354,-0.3813],[356,-0.3668],[358,-0.347],[360,-0.3481],[362,-0.3397],[364,-0.3263],[366,-0.2993],[368,-0.2778],[370,-0.2593],[372,-0.2172],[374,-0.1382],[376,-0.1235],[378,-0.0279],[380,0.0999],[382,0.2183],[384,-0.0347],[386,-0.423],[388,-0.9419],[390,-1.3603],[392,-2.0677],[394,-2.3127],[396,-2.1616],[398,-1.8214],[400,-1.7222],[402,-1.5583],[404,-1.4682],[406,-1.1794],[408,-0.958],[410,-0.7427],[412,-0.5992],[414,-0.4769],[416,-0.3437],[418,-0.2587],[420,-0.2275],[422,-0.1677],[424,-0.1331],[426,-0.0841],[428,-0.0477],[430,-0.0426],[432,-0.0169],[434,0.0127],[436,0.0303],[438,0.0379],[440,0.0504],[442,0.0619],[444,0.0728],[446,0.057],[448,0.0624],[450,0.0438],[452,0.0193],[454,-0.0189],[456,-0.0714],[458,-0.1037],[460,-0.1476],[462,-0.2332],[464,-0.312],[466,-0.4285],[468,-0.3578],[470,0.2133],[472,0.5487],[474,0.5128],[476,0.5362],[478,1.0621],[480,1.8151],[482,2.7253],[484,2.8943],[486,2.8814],[488,2.5618],[492,1.8184],[494,1.7122],[496,1.593],[500,1.2259],[502,1.0777],[504,0.8977],[506,0.7791],[508,0.6497],[510,0.5511],[512,0.4737],[514,0.3878],[516,0.3196],[518,0.3111],[520,0.2207],[522,0.1761],[524,0.1378],[526,-0.0224],[528,-0.0192],[530,-0.0171],[532,-0.0023],[534,0.1323],[536,0.0984],[538,0.1508],[540,0.1734],[542,0.0946],[544,0.0683],[546,0.0808],[548,0.1055],[550,0.1418],[552,0.2019],[554,0.2512],[556,0.2542],[558,0.2532],[560,0.253],[562,0.2556],[564,0.2727],[566,0.2612],[568,0.2586],[570,0.2487],[572,0.2607],[574,0.2605],[576,0.257],[578,0.2457],[580,0.2423],[582,0.2718],[584,0.5425],[586,0.7078],[588,0.7018],[590,0.7433],[592,0.6427],[594,0.6187],[596,0.4571],[598,0.3631],[600,0.2479],[602,0.2505],[604,0.2633],[606,0.2602],[608,0.2437],[612,0.1878],[614,0.1708],[616,0.1606],[618,0.1551],[620,0.1538],[622,0.1499],[624,0.145],[626,0.1462],[628,0.1582],[630,0.1779],[632,0.2029],[634,0.2259],[636,0.2259],[638,0.2306],[640,0.2272],[642,0.2101],[644,0.2309],[646,0.2286],[648,0.2109],[650,0.2031],[652,0.2046],[656,0.1956],[658,0.204],[660,0.2096],[662,0.2045],[664,0.2043],[666,0.2142],[668,0.2276],[670,0.2308],[672,0.2291],[674,0.1902],[676,0.1823],[678,0.1769],[680,0.177],[682,0.1833],[684,0.1892],[686,0.1746],[690,0.1883],[692,0.1961],[694,0.1912],[696,0.2004],[698,0.2587],[700,0.265],[702,0.2563],[704,0.2441],[706,0.2412],[708,0.2391],[710,0.2355],[712,0.2347],[714,0.2276],[716,0.2137],[718,0.2212],[720,0.2212]]
       }
       CTRL_R_eyeAim.Location.Z {
        type=MovieSceneScriptingFloatChannel
        default=0
        thinned_from=492
        keys [[-262,-0.5387],[-260,-0.5397],[-258,-0.5273],[-256,-0.5385],[-254,-0.5217],[-252,-0.4967],[-250,-0.4858],[-248,-0.4935],[-244,-0.4937],[-242,-0.4768],[-240,-0.4622],[-238,-0.4654],[-236,-0.433],[-234,-0.4336],[-232,-0.4395],[-230,-0.4349],[-228,-0.4521],[-226,-0.4792],[-224,-0.5018],[-222,-0.5202],[-220,-0.5328],[-218,-0.5424],[-216,-0.5328],[-212,-0.5336],[-210,-0.5266],[-208,-0.5584],[-206,-0.544],[-204,-0.5747],[-202,-0.535],[-200,-0.5495],[-198,-0.5466],[-196,-0.55],[-194,-0.5633],[-192,-0.5854],[-190,-0.5913],[-188,-0.5802],[-186,-0.5793],[-184,-0.5664],[-182,-0.5623],[-180,-0.5553],[-178,-0.5413],[-176,-0.5394],[-174,-0.5019],[-172,-0.4714],[-170,-0.4579],[-168,-0.4534],[-166,-0.4433],[-164,-0.4307],[-162,-0.4465],[-160,-0.4481],[-158,-0.4425],[-156,-0.3958],[-154,-0.42],[-152,-0.4385],[-150,-0.3932],[-148,-0.4263],[-146,-0.4632],[-144,-0.443],[-142,-0.4573],[-140,-0.4301],[-138,-0.432],[-136,-0.4328],[-134,-0.4165],[-132,-0.4273],[-130,-0.4302],[-128,-0.4227],[-126,-0.4022],[-124,-0.4122],[-122,-0.4326],[-120,-0.4376],[-118,-0.4297],[-116,-0.4448],[-112,-0.435],[-110,-0.434],[-108,-0.4182],[-106,-0.4231],[-104,-0.4306],[-102,-0.454],[-100,-0.4473],[-98,-0.446],[-96,-0.4088],[-94,-0.4115],[-92,-0.4094],[-90,-0.4087],[-88,-0.4062],[-86,-0.3959],[-84,-0.3993],[-82,-0.4225],[-80,-0.4444],[-78,-0.4466],[-76,-0.4435],[-74,-0.4305],[-72,-0.4542],[-70,-0.4451],[-68,-0.4564],[-66,-0.4445],[-64,-0.4404],[-62,-0.4384],[-60,-0.406],[-58,-0.3868],[-56,-0.3632],[-54,-0.3987],[-52,-0.3936],[-50,-0.3716],[-48,-0.3654],[-46,-0.3464],[-44,0.6844],[-42,0.5977],[-40,-0.461],[-38,-0.5615],[-36,-0.5267],[-34,-0.4502],[-32,-0.3945],[-30,-0.3445],[-28,-0.3202],[-26,-0.3242],[-24,-0.2982],[-22,-0.3007],[-20,-0.3013],[-18,-0.2819],[-16,-0.2963],[-14,-0.2782],[-12,-0.2589],[-10,-0.265],[-8,-0.2469],[-6,-0.2497],[-4,-0.2351],[-2,-0.2456],[0,-0.2339],[2,-0.2362],[4,-0.2616],[6,-0.2817],[8,-0.294],[10,-0.2622],[12,-0.2598],[16,-0.2596],[18,-0.2582],[20,-0.2577],[22,-0.2664],[24,-0.2699],[26,-0.2616],[28,-0.2545],[30,-0.2685],[32,-0.2831],[34,-0.2844],[36,-0.294],[38,-0.3084],[40,-0.3179],[42,-0.3382],[44,-0.3024],[46,-0.2977],[48,-0.2878],[50,-0.2876],[52,-0.2952],[54,-0.3107],[56,-0.3271],[58,-0.3259],[60,-0.3198],[62,-0.3059],[64,-0.282],[66,-0.2659],[68,-0.2815],[70,-0.2787],[72,-0.2628],[74,-0.2834],[76,-0.2865],[78,-0.2947],[80,-0.2955],[82,-0.2746],[84,-0.2604],[86,-0.2545],[88,-0.2539],[90,-0.258],[92,-0.2525],[94,-0.2754],[96,-0.281],[98,-0.317],[100,-0.3464],[102,-0.3714],[104,-0.4204],[106,-0.4717],[108,-0.4872],[110,-0.5249],[112,-0.5257],[114,-0.5536],[116,-0.563],[118,-0.5705],[120,-0.5732],[122,-0.5562],[124,-0.5817],[126,-0.583],[128,-0.6036],[130,-0.6036],[132,-0.6347],[136,-0.6388],[138,-0.6612],[140,-0.6355],[142,-0.6419],[144,-0.6216],[146,-0.6171],[148,-0.5948],[150,-0.5785],[152,-0.5429],[154,-0.5359],[156,-0.5364],[158,-0.5454],[160,-0.5265],[162,-0.558],[164,-0.5201],[168,-0.5223],[170,-0.517],[172,-0.526],[174,-0.5187],[176,-0.5416],[178,-0.5167],[180,-0.5237],[182,-0.513],[184,-0.5328],[186,-0.5235],[188,-0.5015],[190,-0.4964],[192,-0.4978],[194,-0.5053],[196,-0.4946],[198,-0.5033],[200,-0.5236],[202,-0.5128],[204,-0.5066],[206,-0.5121],[208,-0.5256],[210,-0.5265],[212,-0.5028],[214,-0.52],[216,-0.5021],[218,-0.4613],[220,-0.4731],[222,-0.4597],[224,-0.4384],[226,-0.4285],[228,-0.4521],[230,-0.4627],[232,-0.4641],[234,-0.4738],[236,-0.4628],[238,-0.4422],[240,-0.4454],[242,-0.4618],[244,-0.4308],[246,-0.3717],[248,-0.3676],[250,-0.3579],[252,-0.3728],[254,-0.3843],[256,-0.378],[258,-0.3977],[260,-0.3814],[262,-0.3821],[264,-0.3792],[266,-0.3866],[268,-0.4056],[270,-0.4032],[272,-0.4226],[274,-0.3751],[276,-0.3588],[278,-0.2431],[280,-0.1784],[282,-0.1429],[284,-0.1287],[286,-0.1292],[288,-0.1727],[290,-0.1998],[292,-0.2098],[294,-0.1978],[296,-0.1732],[298,-0.1973],[300,-0.2003],[302,-0.1906],[304,-0.1899],[306,-0.2242],[308,-0.2291],[310,-0.2419],[312,-0.2411],[314,-0.242],[316,-0.23],[318,-0.208],[320,-0.2021],[322,-0.1949],[324,-0.1827],[326,-0.2119],[328,-0.1741],[330,-0.1788],[332,-0.1629],[334,-0.163],[336,-0.1614],[338,-0.1927],[340,-0.1941],[342,-0.1752],[344,-0.1771],[346,-0.195],[348,-0.2101],[350,-0.2144],[352,-0.2409],[354,-0.2379],[356,-0.2278],[358,-0.2248],[360,-0.2097],[362,-0.2147],[364,-0.2075],[366,-0.2284],[368,-0.2292],[370,-0.2242],[372,-0.2332],[374,-0.2567],[376,-0.26],[378,-0.2416],[380,-0.2571],[382,-0.2527],[384,-0.5156],[386,-0.4083],[388,-0.851],[390,-0.55],[392,0.2074],[394,0.4932],[396,0.3551],[398,0.1092],[400,0.0461],[402,-0.064],[404,-0.0801],[406,-0.1112],[408,-0.1699],[410,-0.2629],[412,-0.3233],[414,-0.3946],[416,-0.4215],[418,-0.4294],[420,-0.3864],[422,-0.3738],[424,-0.3731],[426,-0.3918],[428,-0.4042],[430,-0.4053],[432,-0.4215],[434,-0.4218],[436,-0.4156],[438,-0.3973],[440,-0.4017],[442,-0.387],[444,-0.3877],[446,-0.3837],[448,-0.3866],[450,-0.3832],[452,-0.3812],[454,-0.3751],[456,-0.3494],[458,-0.3354],[460,-0.3243],[462,-0.3155],[464,-0.2885],[466,-0.3148],[468,-0.3712],[470,-0.5021],[472,0.0169],[474,1.1873],[476,1.5733],[478,2.0164],[480,1.8286],[482,0.9429],[484,0.7802],[486,0.4325],[488,0.227],[490,0.0743],[492,-0.1132],[494,-0.1242],[496,-0.2162],[498,-0.3221],[500,-0.3713],[502,-0.4252],[504,-0.4264],[506,-0.4251],[508,-0.4566],[510,-0.4381],[512,-0.4299],[514,-0.5176],[516,-0.6344],[518,-0.631],[520,-0.3847],[522,0.2541],[524,0.7936],[526,0.7248],[528,0.7691],[530,0.7759],[532,0.5977],[534,0.6912],[536,0.8849],[538,-0.1502],[540,-0.3664],[542,-0.4646],[544,-0.4804],[546,-0.4748],[548,-0.4746],[550,-0.4603],[552,-0.4483],[554,-0.4592],[556,-0.4571],[558,-0.4521],[560,-0.4346],[562,-0.437],[564,-0.4404],[566,-0.4326],[568,-0.4319],[570,-0.4414],[572,-0.4557],[574,-0.4513],[576,-0.454],[578,-0.4787],[580,-0.4907],[582,-0.4923],[584,-0.5049],[586,0.0587],[588,-0.2057],[590,-0.6828],[592,-0.6648],[594,-0.6537],[596,-0.6237],[598,-0.6121],[600,-0.6231],[602,-0.623],[604,-0.6417],[606,-0.6484],[608,-0.6649],[610,-0.6665],[612,-0.6802],[614,-0.6999],[616,-0.7053],[618,-0.7229],[620,-0.7311],[622,-0.7171],[624,-0.7074],[626,-0.7058],[628,-0.7159],[630,-0.7132],[632,-0.7086],[634,-0.6993],[636,-0.7067],[640,-0.7099],[642,-0.7213],[644,-0.7039],[646,-0.6744],[648,-0.6883],[650,-0.692],[652,-0.6706],[654,-0.6741],[656,-0.6433],[658,-0.6545],[660,-0.6682],[662,-0.6531],[664,-0.6355],[666,-0.6081],[670,-0.6089],[672,-0.5975],[674,-0.596],[676,-0.5831],[678,-0.6007],[680,-0.5963],[682,-0.5853],[684,-0.6027],[686,-0.588],[688,-0.581],[690,-0.5897],[692,-0.6045],[694,-0.6087],[696,-0.6],[698,-0.6237],[700,-0.6149],[702,-0.6329],[704,-0.6177],[706,-0.613],[708,-0.6051],[710,-0.6074],[714,-0.6079],[716,-0.5978],[718,-0.595],[720,-0.595]]
       }
       CTRL_R_eyeAim.Rotation.X {
        type=MovieSceneScriptingFloatChannel
        default=0
        thinned_from=492
        keys [[-262,-1.2714],[-260,-1.272],[-258,-1.2425],[-256,-1.2687],[-254,-1.2269],[-252,-1.167],[-250,-1.1408],[-248,-1.1588],[-244,-1.1593],[-242,-1.1195],[-240,-1.0853],[-238,-1.0928],[-236,-1.0169],[-234,-1.0184],[-232,-1.0328],[-230,-1.0223],[-228,-1.064],[-226,-1.1286],[-224,-1.1824],[-222,-1.2263],[-220,-1.2572],[-218,-1.2799],[-216,-1.2568],[-212,-1.258],[-210,-1.2395],[-208,-1.3137],[-206,-1.279],[-204,-1.3516],[-202,-1.258],[-200,-1.2914],[-198,-1.2844],[-196,-1.2921],[-194,-1.3228],[-192,-1.3747],[-190,-1.3885],[-188,-1.3623],[-186,-1.3603],[-184,-1.3298],[-182,-1.3203],[-180,-1.3038],[-178,-1.2711],[-176,-1.2668],[-174,-1.1791],[-172,-1.1081],[-170,-1.0761],[-168,-1.0651],[-166,-1.0411],[-164,-1.0112],[-162,-1.0484],[-160,-1.0522],[-158,-1.0391],[-156,-0.9295],[-154,-0.9873],[-152,-1.0308],[-150,-0.9248],[-148,-1.0023],[-146,-1.0886],[-144,-1.0408],[-142,-1.0742],[-140,-1.0098],[-138,-1.0144],[-136,-1.0163],[-134,-0.9779],[-132,-1.0033],[-130,-1.0106],[-128,-0.9927],[-126,-0.945],[-124,-0.9684],[-122,-1.0162],[-120,-1.0278],[-118,-1.0097],[-116,-1.0452],[-112,-1.022],[-110,-1.0194],[-108,-0.9821],[-106,-0.9937],[-104,-1.0112],[-102,-1.0661],[-100,-1.0503],[-98,-1.0472],[-96,-0.9599],[-94,-0.9663],[-92,-0.9613],[-90,-0.9596],[-88,-0.9538],[-86,-0.9296],[-84,-0.9375],[-82,-0.992],[-80,-1.0433],[-78,-1.0485],[-76,-1.0412],[-74,-1.0109],[-72,-1.0666],[-70,-1.0456],[-68,-1.0726],[-66,-1.0452],[-64,-1.0359],[-62,-1.0315],[-60,-0.9557],[-58,-0.9107],[-56,-0.8559],[-54,-0.9391],[-52,-0.9266],[-50,-0.8746],[-48,-0.8598],[-46,-0.8149],[-44,1.5921],[-42,1.3877],[-40,-1.0904],[-38,-1.3207],[-36,-1.2382],[-34,-1.0585],[-32,-0.9275],[-30,-0.8097],[-28,-0.7522],[-26,-0.7617],[-24,-0.7006],[-22,-0.7063],[-20,-0.7077],[-18,-0.6619],[-16,-0.6957],[-14,-0.6533],[-12,-0.6081],[-10,-0.6223],[-8,-0.5797],[-6,-0.5862],[-4,-0.552],[-2,-0.5767],[0,-0.5493],[2,-0.5548],[4,-0.6144],[6,-0.6619],[8,-0.6907],[10,-0.6163],[12,-0.6107],[16,-0.6102],[18,-0.6067],[20,-0.6055],[22,-0.6259],[24,-0.6341],[26,-0.6148],[28,-0.598],[30,-0.6305],[32,-0.6647],[34,-0.668],[36,-0.6906],[38,-0.7243],[40,-0.7466],[42,-0.7942],[44,-0.7102],[46,-0.699],[48,-0.6758],[50,-0.6752],[52,-0.6933],[54,-0.7298],[56,-0.7683],[58,-0.7654],[60,-0.751],[62,-0.7182],[64,-0.6622],[66,-0.6242],[68,-0.661],[70,-0.6543],[72,-0.617],[74,-0.6655],[76,-0.6727],[78,-0.6918],[80,-0.6937],[82,-0.6447],[84,-0.6114],[86,-0.5978],[88,-0.5968],[90,-0.6062],[92,-0.5941],[94,-0.6477],[96,-0.6612],[98,-0.7453],[100,-0.8145],[102,-0.8733],[104,-0.9891],[106,-1.1106],[108,-1.1473],[110,-1.2354],[112,-1.2362],[114,-1.3005],[116,-1.322],[118,-1.3397],[120,-1.346],[122,-1.3065],[124,-1.3664],[126,-1.3694],[128,-1.4175],[130,-1.4175],[132,-1.4909],[136,-1.5003],[138,-1.553],[140,-1.4926],[142,-1.5074],[144,-1.4597],[146,-1.449],[148,-1.3967],[150,-1.3585],[152,-1.2748],[154,-1.2582],[156,-1.2596],[158,-1.2806],[160,-1.2362],[162,-1.3101],[164,-1.2212],[168,-1.2264],[170,-1.2139],[172,-1.2351],[174,-1.2178],[176,-1.2716],[178,-1.2134],[180,-1.2298],[182,-1.2054],[184,-1.2526],[186,-1.231],[188,-1.1801],[190,-1.1684],[192,-1.1715],[194,-1.189],[196,-1.1635],[198,-1.1844],[200,-1.2334],[202,-1.2084],[204,-1.1938],[206,-1.2077],[208,-1.2406],[210,-1.2428],[212,-1.1896],[214,-1.2317],[216,-1.1928],[218,-1.1005],[220,-1.1301],[222,-1.1018],[224,-1.0544],[226,-1.0375],[228,-1.0936],[230,-1.1253],[232,-1.1353],[234,-1.1636],[236,-1.1473],[238,-1.1065],[240,-1.1231],[242,-1.1666],[244,-1.0942],[246,-0.963],[248,-0.9581],[250,-0.9347],[252,-0.9728],[254,-1.0008],[256,-0.985],[258,-1.0291],[260,-0.9931],[262,-0.9938],[264,-0.985],[266,-1.0078],[268,-1.0664],[270,-1.0706],[272,-1.123],[274,-1.0393],[276,-1.0213],[278,-0.7327],[280,-0.5867],[282,-0.5021],[284,-0.4729],[286,-0.4635],[288,-0.5554],[290,-0.6098],[292,-0.6282],[294,-0.6031],[296,-0.5433],[298,-0.6052],[300,-0.6169],[302,-0.599],[304,-0.5937],[306,-0.6749],[308,-0.6864],[310,-0.7038],[312,-0.6976],[314,-0.6959],[316,-0.6602],[318,-0.6046],[320,-0.5821],[322,-0.561],[324,-0.5252],[326,-0.5949],[328,-0.5059],[330,-0.5142],[332,-0.4705],[334,-0.4666],[336,-0.4553],[338,-0.5274],[340,-0.527],[342,-0.4783],[344,-0.4784],[346,-0.5132],[348,-0.5406],[350,-0.545],[352,-0.6052],[354,-0.5973],[356,-0.5702],[358,-0.5596],[360,-0.5242],[362,-0.5342],[364,-0.5151],[366,-0.5597],[368,-0.5583],[370,-0.5439],[372,-0.5597],[374,-0.6077],[376,-0.6142],[378,-0.5673],[380,-0.6058],[382,-0.6043],[384,-1.211],[386,-1.0026],[388,-2.2439],[390,-1.806],[392,-0.7172],[394,-0.3165],[396,-0.3637],[398,-0.5408],[400,-0.5954],[402,-0.7137],[404,-0.6706],[406,-0.5741],[408,-0.6035],[410,-0.7391],[412,-0.8381],[414,-0.9768],[416,-1.0159],[418,-1.0232],[420,-0.9188],[422,-0.884],[424,-0.8802],[426,-0.9216],[428,-0.9497],[430,-0.952],[432,-0.9897],[434,-0.9905],[436,-0.976],[438,-0.9332],[440,-0.9437],[442,-0.9093],[444,-0.9113],[446,-0.9015],[448,-0.9083],[450,-0.9002],[452,-0.8951],[454,-0.8808],[456,-0.8215],[458,-0.7899],[460,-0.7664],[462,-0.753],[464,-0.6993],[466,-0.7815],[468,-0.9014],[470,-1.1887],[472,-0.028],[474,2.7363],[476,3.6441],[478,4.5228],[480,3.5968],[482,0.5128],[484,-0.1039],[486,-0.8955],[488,-0.9521],[490,-0.8962],[492,-0.9988],[494,-0.9398],[498,-1.1939],[500,-1.2027],[502,-1.2531],[504,-1.1783],[506,-1.1322],[508,-1.1654],[510,-1.0958],[512,-1.059],[514,-1.2474],[516,-1.5103],[518,-1.5011],[520,-0.9142],[522,0.5884],[524,1.8596],[526,1.7022],[528,1.8064],[530,1.8222],[532,1.4038],[534,1.6196],[536,2.0765],[538,-0.3577],[540,-0.8668],[542,-1.0929],[544,-1.1288],[546,-1.1162],[548,-1.1166],[550,-1.0851],[552,-1.0615],[554,-1.0922],[556,-1.0875],[558,-1.0756],[560,-1.0346],[562,-1.0404],[564,-1.0505],[566,-1.0308],[568,-1.0288],[570,-1.05],[572,-1.0849],[574,-1.0746],[576,-1.0804],[578,-1.1371],[580,-1.1649],[582,-1.1721],[584,-1.2506],[586,0.0243],[588,-0.5943],[590,-1.7254],[592,-1.6519],[594,-1.6191],[596,-1.509],[598,-1.4646],[600,-1.4755],[602,-1.4755],[604,-1.5205],[606,-1.536],[608,-1.5729],[610,-1.574],[612,-1.604],[614,-1.6488],[616,-1.6609],[618,-1.7019],[620,-1.7211],[622,-1.6881],[624,-1.665],[626,-1.6613],[628,-1.6857],[630,-1.6807],[632,-1.6716],[634,-1.6519],[636,-1.6693],[638,-1.6751],[640,-1.6769],[642,-1.702],[644,-1.6631],[646,-1.5936],[648,-1.6248],[650,-1.6327],[652,-1.5828],[654,-1.5906],[656,-1.5179],[658,-1.5449],[660,-1.5774],[662,-1.5417],[664,-1.5004],[666,-1.4369],[670,-1.4402],[672,-1.4134],[674,-1.4066],[676,-1.3757],[678,-1.4166],[680,-1.4062],[682,-1.3809],[684,-1.4221],[686,-1.3865],[688,-1.3706],[690,-1.3915],[692,-1.427],[694,-1.4364],[696,-1.4167],[698,-1.4778],[700,-1.4578],[702,-1.4992],[704,-1.4622],[706,-1.4511],[708,-1.4323],[710,-1.4372],[714,-1.4377],[716,-1.4126],[718,-1.4068],[720,-1.4068]]
       }
       CTRL_R_eyeAim.Rotation.Y {
        type=MovieSceneScriptingFloatChannel
        default=0
        thinned_from=492
        keys [[-262,0.3944],[-260,0.3616],[-258,0.3589],[-256,0.3632],[-254,0.3325],[-252,0.2982],[-250,0.2838],[-248,0.2757],[-244,0.2773],[-242,0.2634],[-240,0.2489],[-238,0.2502],[-236,0.2336],[-234,0.2337],[-232,0.2349],[-230,0.2296],[-228,0.2306],[-226,0.2347],[-224,0.2315],[-222,0.2299],[-220,0.2342],[-218,0.2478],[-214,0.2528],[-212,0.2532],[-210,0.2523],[-208,0.256],[-206,0.2371],[-204,0.2427],[-202,0.2163],[-200,0.2108],[-198,0.1999],[-196,0.1887],[-194,0.1831],[-190,0.181],[-188,0.1681],[-186,0.1673],[-184,0.1645],[-182,0.169],[-180,0.172],[-178,0.1797],[-176,0.1901],[-174,0.1905],[-172,0.1894],[-170,0.1931],[-168,0.1999],[-164,0.2006],[-162,0.2085],[-160,0.2098],[-158,0.2078],[-156,0.1908],[-154,0.2071],[-152,0.2135],[-150,0.1967],[-148,0.2031],[-146,0.2117],[-144,0.1954],[-142,0.1964],[-140,0.1884],[-138,0.1885],[-136,0.189],[-134,0.1844],[-132,0.1904],[-130,0.1945],[-128,0.1913],[-126,0.1874],[-124,0.1916],[-122,0.2016],[-120,0.2046],[-118,0.1992],[-116,0.2111],[-114,0.2087],[-112,0.2057],[-110,0.2058],[-108,0.1969],[-106,0.1988],[-104,0.2036],[-102,0.2158],[-100,0.2177],[-98,0.2176],[-96,0.209],[-94,0.2206],[-92,0.2208],[-90,0.2244],[-88,0.2238],[-86,0.2211],[-84,0.2275],[-82,0.242],[-80,0.2486],[-78,0.2413],[-76,0.2371],[-74,0.2244],[-72,0.2322],[-70,0.2257],[-68,0.2331],[-66,0.2252],[-64,0.222],[-62,0.2221],[-60,0.2078],[-58,0.1984],[-56,0.1868],[-54,0.206],[-52,0.2029],[-50,0.187],[-48,0.1812],[-46,0.1035],[-44,-0.1353],[-42,-0.2168],[-40,0.1931],[-38,0.2344],[-36,0.2247],[-34,0.2011],[-32,0.1835],[-30,0.1611],[-28,0.1474],[-26,0.1493],[-24,0.1385],[-22,0.1366],[-20,0.1393],[-18,0.1324],[-16,0.1382],[-14,0.1326],[-12,0.124],[-10,0.1341],[-8,0.1256],[-6,0.1295],[-4,0.1227],[-2,0.1281],[0,0.119],[2,0.1203],[4,0.131],[6,0.1425],[8,0.1484],[10,0.1343],[12,0.1343],[14,0.1354],[16,0.1352],[18,0.1342],[20,0.1359],[22,0.1422],[24,0.1447],[26,0.1407],[28,0.1387],[30,0.147],[32,0.1545],[34,0.1555],[36,0.1627],[38,0.1716],[40,0.1788],[42,0.1872],[44,0.1702],[46,0.168],[48,0.1628],[50,0.162],[52,0.1688],[54,0.1787],[56,0.1895],[58,0.1904],[60,0.1901],[62,0.1869],[64,0.1743],[66,0.1646],[68,0.1755],[70,0.1743],[72,0.168],[74,0.178],[76,0.1775],[78,0.182],[80,0.1853],[82,0.1728],[84,0.1622],[86,0.1604],[88,0.1615],[90,0.1638],[92,0.1585],[94,0.1697],[96,0.1651],[98,0.1797],[100,0.1862],[102,0.1763],[104,0.1822],[106,0.1567],[108,0.1561],[110,0.1457],[112,0.1226],[114,0.1161],[116,0.1164],[118,0.1196],[120,0.1205],[122,0.115],[124,0.1184],[126,0.1178],[128,0.1086],[130,0.1086],[132,0.1038],[134,0.1036],[138,0.1121],[140,0.121],[142,0.131],[144,0.1363],[146,0.1647],[148,0.1741],[150,0.1791],[152,0.1853],[154,0.2],[156,0.2007],[158,0.2118],[160,0.2153],[162,0.2356],[164,0.2218],[166,0.2273],[168,0.2336],[170,0.236],[172,0.2409],[174,0.2399],[176,0.258],[178,0.2492],[180,0.2492],[182,0.2454],[184,0.2488],[186,0.2443],[188,0.2343],[190,0.2316],[192,0.2263],[194,0.2289],[196,0.2208],[198,0.2204],[200,0.231],[202,0.2303],[204,0.2288],[206,0.2332],[208,0.2456],[210,0.2464],[212,0.2373],[214,0.2482],[216,0.2451],[218,0.2301],[220,0.2442],[222,0.2387],[224,0.2304],[226,0.2296],[228,0.2443],[230,0.2636],[232,0.2684],[234,0.2766],[236,0.2758],[238,0.2721],[240,0.2806],[242,0.3039],[244,0.2934],[246,0.2729],[250,0.27],[252,0.2822],[254,0.2882],[256,0.2849],[258,0.2956],[260,0.2851],[262,0.2865],[264,0.2778],[266,0.2731],[268,0.286],[270,0.2949],[272,0.344],[274,0.3776],[276,0.4569],[278,0.377],[280,0.333],[282,0.295],[284,0.2894],[286,0.2816],[288,0.3333],[290,0.3622],[292,0.3736],[294,0.354],[296,0.3223],[298,0.3558],[300,0.3644],[302,0.3513],[304,0.3464],[306,0.3925],[310,0.4041],[314,0.4],[316,0.3776],[318,0.3401],[320,0.3341],[322,0.3245],[324,0.3036],[326,0.3452],[328,0.2993],[330,0.3035],[332,0.2808],[334,0.2855],[336,0.2872],[338,0.3317],[340,0.3294],[342,0.288],[344,0.2776],[346,0.2827],[348,0.2848],[350,0.2748],[352,0.2961],[354,0.2893],[356,0.2682],[358,0.2621],[360,0.2426],[362,0.2456],[364,0.2366],[366,0.2566],[368,0.2542],[370,0.2437],[372,0.2443],[374,0.2468],[376,0.2461],[378,0.215],[380,0.2096],[382,0.1952],[384,0.3898],[386,0.3472],[388,1.0542],[390,0.894],[392,0.3583],[394,0.1459],[396,0.1365],[398,0.1591],[400,0.1643],[402,0.1714],[404,0.1035],[406,0.11],[408,0.1054],[410,0.1061],[412,0.1031],[414,0.1082],[416,0.1011],[418,0.0937],[420,0.1073],[422,0.1111],[424,0.1058],[426,0.0982],[428,0.0897],[430,0.0834],[432,0.0817],[434,0.0777],[436,0.0779],[438,0.0754],[440,0.0758],[442,0.077],[444,0.0798],[446,0.0805],[448,0.0829],[450,0.0848],[454,0.0934],[456,0.096],[458,0.1043],[460,0.1147],[462,0.1339],[464,0.1426],[466,0.1966],[468,0.2401],[470,0.2806],[472,0.0061],[474,-0.4098],[476,-0.2574],[478,-0.6189],[480,-0.4158],[482,-0.0182],[484,0.002],[486,0.0087],[488,0.0349],[490,0.0638],[492,0.1133],[494,0.1088],[496,0.1328],[498,0.1741],[500,0.2009],[502,0.2149],[504,0.2226],[506,0.2397],[508,0.2584],[510,0.2578],[512,0.2559],[514,0.2336],[516,-0.0084],[518,-0.047],[520,-0.2814],[522,0.2948],[524,1.0927],[526,1.0352],[528,1.0959],[530,1.0857],[532,0.8233],[534,0.5253],[536,0.289],[538,0.0721],[540,0.234],[542,0.478],[544,0.4601],[546,0.4157],[548,0.3471],[550,0.3056],[552,0.3012],[554,0.3096],[556,0.2999],[558,0.295],[560,0.2812],[564,0.2875],[566,0.2843],[570,0.2838],[572,0.2996],[574,0.2968],[576,0.2987],[578,0.3151],[580,0.3223],[582,0.3146],[584,0.3093],[586,-0.0057],[588,0.1435],[590,0.4401],[592,0.4207],[594,0.4084],[596,0.303],[598,0.2424],[600,0.162],[602,0.1571],[604,0.1219],[606,0.1166],[608,0.0991],[610,0.0871],[612,0.0741],[614,0.0663],[616,0.0602],[618,0.0571],[620,0.0562],[622,0.0513],[624,0.0495],[626,0.0497],[628,0.0536],[630,0.0537],[632,0.0626],[634,0.0647],[636,0.0726],[638,0.0745],[640,0.0782],[642,0.0718],[644,0.0731],[646,0.0718],[648,0.0679],[650,0.0688],[652,0.0714],[654,0.0765],[656,0.0699],[660,0.0836],[662,0.0851],[664,0.0829],[666,0.0894],[668,0.0952],[670,0.0997],[672,0.1004],[674,0.0993],[676,0.1016],[678,0.1027],[680,0.1065],[682,0.1041],[684,0.1081],[686,0.1012],[688,0.1033],[690,0.1015],[692,0.1069],[694,0.1065],[696,0.1077],[698,0.1191],[700,0.1207],[702,0.1262],[704,0.1255],[706,0.1274],[710,0.1323],[712,0.1321],[716,0.1375],[718,0.1406],[720,0.1406]]
       }
       CTRL_R_eyeAim.Rotation.Z {
        type=MovieSceneScriptingFloatChannel
        default=0
        thinned_from=492
        keys [[-262,-0.0587],[-260,-0.0745],[-258,-0.0801],[-256,-0.0815],[-254,-0.0701],[-252,-0.0382],[-250,-0.0278],[-248,-0.0384],[-246,-0.0408],[-244,-0.0411],[-242,-0.0439],[-240,-0.0972],[-238,-0.1101],[-236,-0.0671],[-234,-0.0678],[-232,-0.0772],[-230,-0.071],[-228,-0.0602],[-226,-0.0733],[-224,-0.0671],[-222,-0.05],[-220,-0.0443],[-218,-0.0607],[-216,-0.0365],[-214,-0.0599],[-212,-0.0643],[-210,-0.0506],[-208,-0.0726],[-206,-0.0782],[-204,-0.0801],[-200,-0.0805],[-198,-0.0622],[-196,-0.0746],[-194,-0.0616],[-192,-0.088],[-190,-0.0987],[-188,-0.072],[-186,-0.0704],[-184,-0.0549],[-182,-0.0923],[-180,-0.0769],[-178,-0.0626],[-176,-0.0851],[-174,-0.0971],[-172,-0.0988],[-170,-0.0965],[-168,-0.1031],[-166,-0.1135],[-164,-0.1181],[-162,-0.1435],[-160,-0.1818],[-158,-0.1828],[-156,-0.1891],[-154,-0.1261],[-152,-0.076],[-150,-0.0512],[-148,-0.018],[-146,-0.0072],[-144,-0.0016],[-142,0.0223],[-140,-0.0203],[-138,0.0461],[-136,0.0179],[-134,0.0311],[-132,0.0352],[-130,0.0125],[-128,0.0245],[-126,0.012],[-124,0.0286],[-122,0.0395],[-120,0.0506],[-118,0.0315],[-116,0.0684],[-114,0.0704],[-112,0.0744],[-110,0.0634],[-108,0.0778],[-106,0.103],[-104,0.1074],[-102,0.0609],[-100,0.0666],[-98,0.0684],[-96,0.0408],[-94,0.0383],[-92,0.0322],[-90,0.0273],[-88,0.0213],[-86,-0.0034],[-84,-0.0421],[-82,-0.0685],[-80,-0.1056],[-78,-0.1262],[-76,-0.1237],[-74,-0.099],[-72,-0.0874],[-70,-0.0284],[-68,-0.0363],[-66,-0.0134],[-64,-0.0057],[-62,0.0067],[-60,0.0416],[-58,0.0409],[-56,0.0387],[-54,0.0377],[-52,0.0377],[-50,0.0575],[-48,0.0652],[-46,0.0505],[-44,-0.0693],[-42,-0.1027],[-40,-0.0621],[-38,-0.0484],[-36,-0.0423],[-34,-0.0157],[-32,-0.0052],[-30,-0.0024],[-28,-0.0291],[-26,-0.0362],[-24,-0.0356],[-22,0.0125],[-20,-0.0171],[-18,-0.032],[-16,-0.0103],[-14,0.0199],[-12,0.0082],[-10,0.0008],[-8,-0.0243],[-6,0.0204],[-4,0.0067],[-2,0.0016],[0,0.0002],[2,-0.0214],[4,-0.0106],[6,-0.0059],[8,-0.0282],[10,0.0114],[12,-0.0343],[14,-0.0379],[16,-0.0462],[18,-0.0676],[20,-0.0654],[22,-0.0476],[24,-0.0428],[26,-0.0514],[28,-0.0708],[30,-0.0593],[32,-0.0625],[34,-0.0522],[36,-0.0656],[38,-0.0499],[40,-0.0436],[42,-0.0413],[44,-0.0404],[46,-0.0476],[48,-0.0688],[50,-0.0763],[52,-0.0783],[54,-0.0668],[58,-0.0569],[60,-0.0634],[62,-0.0536],[64,-0.0439],[66,-0.0649],[68,-0.0486],[70,-0.0478],[72,-0.0463],[74,-0.0589],[76,-0.057],[78,-0.0657],[80,-0.056],[82,-0.0713],[84,-0.0659],[86,-0.0483],[88,-0.0421],[90,-0.0413],[92,-0.039],[94,-0.0154],[96,-0.0273],[98,-0.0109],[100,0.0194],[102,0.0001],[104,-0.0235],[106,-0.0146],[108,-0.0102],[110,-0.0043],[112,-0.019],[114,-0.0169],[116,-0.0254],[118,-0.0364],[120,-0.0376],[122,-0.0122],[124,-0.0128],[126,-0.0143],[128,-0.0377],[130,-0.0377],[132,-0.0393],[134,-0.0148],[136,-0.0051],[138,-0.0027],[140,-0.002],[142,-0.0019],[144,-0.0279],[146,-0.002],[150,-0.0342],[152,0.0138],[154,0.0308],[156,0.0277],[158,-0.0209],[160,-0.0341],[162,-0.0395],[164,-0.0153],[166,-0.0067],[168,-0.004],[170,-0.0021],[172,0.0237],[174,0.0331],[176,0.0351],[178,-0.0072],[180,0.0188],[182,0.0299],[184,0.0279],[186,0.0294],[188,0.0273],[190,0.0077],[192,0.0006],[194,0.0044],[196,0.0141],[198,0.0028],[200,-0.0005],[202,0.0158],[204,0.0236],[206,0.014],[208,0.0263],[210,0.0273],[212,0.0279],[214,0.0104],[216,0.0026],[218,0.0248],[220,0.0065],[222,0.026],[224,0.0075],[226,0.028],[228,0.0336],[230,0.0356],[232,0.0361],[234,0.0352],[236,0.0101],[238,0.002],[240,-0.0014],[242,-0.0028],[244,-0.0027],[248,-0.0021],[250,0.0211],[252,0.0307],[254,-0.0188],[256,0.0186],[258,0.0304],[260,0.0607],[262,0.0248],[264,0.0309],[266,0.0614],[268,0.0433],[270,0.0122],[272,0.0024],[274,0.0256],[276,0.0853],[278,0.051],[280,0.0161],[282,0.0232],[284,0.0102],[286,0.0039],[288,-0.0004],[290,0.0475],[292,-0.0153],[294,-0.0059],[296,0.0258],[298,0.0333],[300,0.036],[302,0.0105],[304,0.0024],[306,-0.0011],[308,-0.0017],[310,0.0236],[312,0.0326],[314,0.0577],[316,0.0704],[318,0.0747],[320,0.0488],[322,0.0413],[324,0.0388],[326,0.0377],[328,0.0132],[330,0.0263],[332,0.0347],[336,0.0387],[338,0.0561],[340,0.0442],[342,0.0392],[344,0.0187],[346,0.0047],[348,0.0239],[350,0.0087],[352,0.0286],[354,0.0238],[356,0.0058],[358,0.0006],[360,-0.0003],[362,-0.0316],[364,-0.0241],[366,-0.0058],[368,0.0172],[370,0.0123],[372,0.0032],[374,0],[378,-0.0008],[380,-0.0037],[382,-0.0261],[384,-0.0704],[386,-0.1016],[388,-0.1801],[390,-0.1731],[392,-0.1594],[394,-0.1562],[396,-0.1186],[398,-0.1075],[400,-0.1112],[402,-0.0904],[404,-0.1053],[406,-0.0866],[408,-0.0539],[410,-0.1001],[412,-0.0988],[414,-0.0618],[416,-0.0734],[418,-0.048],[420,-0.0167],[422,-0.0335],[424,-0.0357],[426,-0.0403],[428,-0.0656],[430,-0.0301],[432,-0.0323],[434,-0.0375],[436,-0.0385],[438,-0.0126],[440,-0.0053],[442,-0.0287],[444,-0.0096],[446,-0.0037],[448,0.0225],[450,0.0329],[452,0.0094],[454,0.0028],[456,-0.0252],[458,-0.0087],[460,-0.0036],[462,-0.0019],[464,-0.0264],[466,-0.0375],[468,-0.04],[470,-0.0718],[472,-0.1416],[474,-0.1784],[476,-0.1918],[478,-0.2417],[480,-0.1914],[482,-0.1198],[484,-0.1096],[486,-0.1168],[488,-0.1175],[490,-0.1171],[492,-0.0992],[494,-0.134],[496,-0.1249],[498,-0.1212],[500,-0.1191],[502,-0.0949],[504,-0.1131],[506,-0.1175],[508,-0.1452],[510,-0.1279],[512,-0.0983],[514,-0.1175],[516,-0.1134],[518,-0.117],[520,-0.173],[522,-0.2397],[524,-0.2494],[526,-0.2579],[528,-0.262],[530,-0.2642],[532,-0.2573],[534,-0.2348],[536,-0.2089],[538,-0.1295],[540,-0.1149],[542,-0.1202],[544,-0.0686],[546,-0.0774],[548,-0.0873],[550,-0.1008],[552,-0.0879],[554,-0.0832],[556,-0.0812],[558,-0.0674],[560,-0.0657],[562,-0.0946],[564,-0.0914],[566,-0.0837],[568,-0.083],[570,-0.0814],[572,-0.0594],[574,-0.0654],[576,-0.0749],[578,-0.0799],[582,-0.0818],[584,-0.123],[586,-0.1767],[588,-0.1729],[590,-0.1227],[592,-0.1178],[594,-0.1171],[596,-0.0794],[598,-0.0591],[600,-0.0523],[602,-0.0495],[604,-0.0151],[606,-0.004],[608,0.0274],[610,0.0347],[612,0.0374],[614,0.038],[616,0.0555],[618,0.0516],[620,0.0454],[622,0.0064],[624,0.0033],[626,0.0006],[628,0.0031],[630,0.051],[632,0.0219],[634,0.0319],[636,0.0099],[638,0.0027],[640,0.0002],[642,-0.0003],[644,0.0234],[646,0.007],[648,0.0016],[650,-0.0001],[652,-0.0007],[654,-0.001],[656,-0.0009],[658,-0.0033],[660,-0.0516],[662,-0.0209],[664,-0.0071],[666,-0.0278],[668,-0.0102],[670,-0.0053],[672,-0.0023],[674,-0.0016],[676,0.0271],[678,0.0094],[680,0.0018],[682,0],[684,-0.0282],[686,-0.0116],[688,-0.004],[690,-0.0022],[692,-0.0029],[694,-0.0272],[696,-0.0104],[698,-0.0264],[700,-0.0132],[702,-0.0049],[704,-0.0028],[706,-0.0333],[708,-0.0383],[710,-0.0401],[712,-0.0403],[714,-0.0123],[716,-0.0053],[718,-0.0297],[720,-0.0297]]
       }
       CTRL_R_eyeAim.Scale.X {type=MovieSceneScriptingFloatChannel, default=1, const [1,492]}
       CTRL_R_eyeAim.Scale.Y {type=MovieSceneScriptingFloatChannel, default=1, const [1,492]}
       CTRL_R_eyeAim.Scale.Z {type=MovieSceneScriptingFloatChannel, default=1, const [1,492]}
       CTRL_convergenceSwitch {type=MovieSceneScriptingFloatChannel, default=0, const [0,492]}
       CTRL_Master.Location.X {type=MovieSceneScriptingFloatChannel, default=0, const [0,492]}
       CTRL_Master.Location.Y {type=MovieSceneScriptingFloatChannel, default=0, const [0,492]}
       CTRL_Master.Location.Z {type=MovieSceneScriptingFloatChannel, default=0, const [0,492]}
       CTRL_Master.Rotation.X {type=MovieSceneScriptingFloatChannel, default=0, const [0,492]}
       CTRL_Master.Rotation.Y {type=MovieSceneScriptingFloatChannel, default=0, const [0,492]}
       CTRL_Master.Rotation.Z {type=MovieSceneScriptingFloatChannel, default=0, const [0,492]}
       CTRL_Master.Scale.X {type=MovieSceneScriptingFloatChannel, default=1, const [1,492]}
       CTRL_Master.Scale.Y {type=MovieSceneScriptingFloatChannel, default=1, const [1,492]}
       CTRL_Master.Scale.Z {type=MovieSceneScriptingFloatChannel, default=1, const [1,492]}
       CTRL_C_jaw.X {
        type=MovieSceneScriptingFloatChannel
        default=0
        thinned_from=492
        keys [[-262,0.0108],[-260,0.0175],[-258,0.0269],[-256,0.0306],[-254,0.0317],[-252,0.0292],[-250,0.0223],[-248,0.0172],[-246,0.0151],[-244,0.0147],[-242,0.0124],[-240,0.0052],[-238,0.0053],[-236,-0.0028],[-234,-0.004],[-232,-0.0119],[-228,-0.0241],[-226,-0.0253],[-224,-0.0234],[-222,-0.02],[-220,-0.0193],[-218,-0.0188],[-216,-0.0202],[-214,-0.0222],[-212,-0.0225],[-210,-0.0238],[-208,-0.0191],[-206,-0.0129],[-204,-0.0106],[-202,-0.008],[-200,-0.0043],[-198,-0.0029],[-196,0.0009],[-194,0.0029],[-192,0.0066],[-190,0.0075],[-188,-0.0022],[-186,-0.0023],[-184,0.0019],[-182,-0.0021],[-180,-0.0035],[-178,-0.0021],[-176,-0.0014],[-174,0.0014],[-172,0.0007],[-170,0.0012],[-168,0.0119],[-166,0.0185],[-164,0.0236],[-162,0.0307],[-160,0.0415],[-158,0.0423],[-156,0.0491],[-154,0.0439],[-152,0.0355],[-150,0.0327],[-148,0.0359],[-146,0.0322],[-144,0.0377],[-142,0.0469],[-138,0.0572],[-136,0.0595],[-134,0.056],[-132,0.049],[-130,0.0467],[-128,0.0487],[-126,0.0482],[-124,0.0487],[-122,0.0518],[-120,0.0556],[-118,0.0523],[-116,0.04],[-114,0.0372],[-112,0.0293],[-110,0.0216],[-108,0.0196],[-106,0.0206],[-104,0.0208],[-102,0.0137],[-100,0.0118],[-98,0.0117],[-96,0.0122],[-94,0.0108],[-92,0.0101],[-90,0.0088],[-88,0.0089],[-86,0.0093],[-82,0.008],[-80,0.012],[-78,0.0207],[-76,0.0233],[-74,0.0295],[-72,0.0403],[-70,0.0569],[-68,0.0624],[-66,0.0611],[-64,0.0505],[-62,0.0451],[-60,0.0192],[-58,0.0177],[-56,0.0143],[-54,0.0125],[-52,0.0096],[-50,0.0059],[-48,0.0047],[-46,-0.0043],[-44,-0.01],[-42,-0.0078],[-40,-0.0036],[-38,-0.0018],[-36,0.0004],[-34,0.0053],[-32,0.0064],[-30,0.0083],[-28,0.0087],[-26,0.0068],[-24,0.008],[-22,0.0102],[-20,0.0087],[-18,0.0086],[-16,0.0072],[-14,0.0063],[-12,0.0097],[-10,0.0107],[-8,0.0098],[-6,0.0098],[-4,0.0102],[-2,0.0101],[2,0.0055],[4,0.0049],[6,0.0068],[8,0.0068],[10,0.0038],[12,-0.0047],[16,-0.0066],[20,-0.0136],[22,-0.0132],[24,-0.0131],[28,-0.0116],[30,-0.0111],[34,-0.0079],[38,-0.0018],[40,-0.0003],[42,-0.0027],[44,-0.0067],[46,-0.0071],[48,-0.0079],[50,-0.0082],[54,-0.0078],[56,-0.0071],[60,-0.0072],[62,-0.0081],[64,-0.0099],[66,-0.0097],[68,-0.0088],[70,-0.0085],[72,-0.006],[74,-0.0057],[76,-0.0058],[78,-0.0054],[80,-0.0059],[82,-0.0049],[84,-0.0044],[86,-0.0034],[88,-0.0032],[90,-0.0034],[92,-0.0084],[94,-0.0107],[96,-0.0082],[98,-0.0092],[100,-0.0074],[104,-0.0001],[106,0.0024],[108,0.0027],[112,0.0047],[114,0.0051],[116,0.003],[120,0.0039],[122,0.0069],[124,0.0083],[126,0.0081],[128,0.0049],[130,0.0049],[132,0.0069],[134,0.0129],[136,0.0157],[138,0.0175],[140,0.0171],[144,0.0171],[146,0.0165],[148,0.0165],[150,0.015],[152,0.0152],[154,0.0166],[156,0.0167],[160,0.0214],[166,0.0245],[168,0.0263],[170,0.0263],[172,0.023],[174,0.0225],[176,0.0244],[178,0.0232],[180,0.0225],[182,0.0231],[184,0.0249],[186,0.0257],[188,0.0275],[190,0.0272],[192,0.0257],[202,0.0244],[204,0.0244],[206,0.0252],[208,0.0272],[210,0.0273],[212,0.0272],[214,0.0269],[216,0.0268],[220,0.024],[222,0.0214],[224,0.0182],[226,0.0156],[228,0.0156],[230,0.0135],[234,0.0043],[236,0.001],[238,-0.0007],[240,0.0004],[242,0.0043],[244,0.0022],[246,-0.0007],[248,0.0017],[250,0.0038],[254,0.0049],[256,0.002],[258,0.0034],[260,0.0078],[262,0.0078],[264,0.0084],[266,0.0107],[268,0.0139],[270,0.0185],[272,0.014],[274,0.0107],[276,0.0109],[278,0.0064],[280,0.0068],[282,0.0064],[284,0.0068],[286,0.0097],[288,0.0112],[292,0.0064],[294,0.0064],[296,0.0078],[298,0.0084],[300,0.0098],[302,0.0101],[304,0.0078],[306,0.0053],[308,0.0039],[312,0.0068],[314,0.0068],[318,0.0078],[320,0.0071],[322,0.004],[324,0.0051],[326,0.0087],[330,0.0073],[332,0.0082],[334,0.0059],[336,0.0054],[338,0.0057],[340,-0.0024],[342,-0.0086],[344,-0.0079],[346,-0.0079],[348,-0.005],[350,-0.0016],[352,0.0028],[354,0.0038],[356,0.0061],[358,0.0042],[362,0.0038],[364,0.0017],[368,0.0028],[370,0.0029],[374,-0.0003],[376,-0.0005],[378,-0.001],[380,0],[382,-0.0007],[384,-0.0081],[386,-0.0157],[388,-0.0112],[390,-0.0079],[392,-0.0058],[394,-0.0088],[396,-0.0084],[398,-0.0018],[400,0.0004],[402,0.0009],[404,-0.0006],[406,-0.0034],[410,-0.0057],[412,-0.0051],[414,-0.0019],[416,0.0002],[418,-0.0005],[420,-0.001],[422,-0.0007],[424,0.0002],[426,0.0034],[428,0.0054],[432,0.0077],[434,0.0087],[436,0.0102],[438,0.0093],[440,0.0102],[442,0.0088],[444,0.0093],[446,0.0112],[448,0.0127],[450,0.0137],[452,0.0127],[454,0.0122],[456,0.0123],[458,0.0137],[460,0.0123],[464,0.0041],[466,-0.0038],[470,-0.0178],[472,-0.0172],[474,-0.0198],[476,-0.0185],[478,-0.013],[480,-0.0039],[482,0.0002],[484,0.0005],[486,-0.0012],[488,0.0029],[490,0.0076],[492,0.0082],[494,0.0194],[496,0.0318],[498,0.0346],[500,0.0312],[502,0.0293],[504,0.0177],[506,0.008],[508,0.0081],[510,0.0032],[512,-0.0022],[514,-0.0068],[516,-0.0124],[518,-0.0124],[520,-0.0085],[522,0.0027],[524,0.0127],[526,0.0147],[528,0.0163],[530,0.0189],[532,0.0199],[534,0.0179],[536,0.017],[538,0.0215],[540,0.0225],[542,0.0193],[544,0.0156],[546,0.0131],[548,0.0038],[550,-0.0026],[552,-0.0016],[554,0.0014],[556,-0.0007],[558,-0.0015],[560,0.0003],[564,0.0009],[566,0.0014],[568,0.0015],[570,0.0031],[572,0.0089],[574,0.0093],[576,0.0089],[578,0.0089],[580,0.0093],[582,0.0091],[584,0.0107],[586,0.0149],[588,0.0171],[590,0.0199],[592,0.0201],[594,0.0197],[596,0.011],[598,0.0068],[600,0.0073],[602,0.0081],[604,0.0142],[606,0.0139],[608,0.0134],[610,0.0142],[614,0.0178],[616,0.0208],[618,0.0233],[620,0.024],[622,0.0268],[624,0.0273],[628,0.0274],[632,0.0285],[634,0.0288],[636,0.0279],[638,0.0288],[640,0.0288],[642,0.0269],[644,0.0269],[646,0.0273],[648,0.0269],[650,0.0279],[652,0.0293],[654,0.0313],[656,0.0332],[658,0.0331],[660,0.0305],[662,0.0288],[664,0.0293],[666,0.028],[668,0.0279],[670,0.0283],[672,0.0278],[674,0.0278],[676,0.0273],[680,0.0254],[684,0.0283],[688,0.0273],[690,0.0273],[694,0.0264],[696,0.0274],[698,0.0289],[700,0.0293],[702,0.0278],[704,0.0283],[706,0.0283],[708,0.0281],[712,0.0283],[714,0.0279],[716,0.0288],[718,0.0304],[720,0.0304]]
       }
       CTRL_C_jaw.Y {
        type=MovieSceneScriptingFloatChannel
        default=0
        thinned_from=492
        keys [[-262,0.2531],[-260,0.2537],[-258,0.2611],[-256,0.2607],[-254,0.2664],[-252,0.2727],[-250,0.2808],[-248,0.2928],[-246,0.2986],[-244,0.2995],[-242,0.3014],[-240,0.3077],[-238,0.3088],[-236,0.3128],[-234,0.3136],[-232,0.3191],[-228,0.322],[-224,0.333],[-222,0.3339],[-220,0.3356],[-218,0.3399],[-216,0.3385],[-214,0.332],[-212,0.3309],[-210,0.3273],[-206,0.3316],[-204,0.3293],[-202,0.3279],[-200,0.3299],[-198,0.331],[-196,0.329],[-194,0.3176],[-192,0.3131],[-190,0.3123],[-188,0.301],[-186,0.3003],[-184,0.2975],[-182,0.2919],[-180,0.2924],[-178,0.2968],[-176,0.3072],[-174,0.3187],[-172,0.3232],[-170,0.3291],[-168,0.3407],[-166,0.3517],[-164,0.3612],[-162,0.3567],[-160,0.3451],[-158,0.3438],[-156,0.3313],[-154,0.2955],[-152,0.2664],[-150,0.2506],[-148,0.2368],[-146,0.2252],[-144,0.2219],[-142,0.2216],[-140,0.2241],[-138,0.2233],[-136,0.2206],[-134,0.2191],[-132,0.2114],[-130,0.1972],[-128,0.1906],[-126,0.1902],[-124,0.181],[-122,0.1611],[-120,0.1345],[-118,0.0972],[-116,0.0635],[-114,0.0601],[-112,0.0524],[-110,0.0483],[-108,0.0462],[-106,0.0448],[-104,0.0442],[-102,0.0456],[-100,0.0477],[-98,0.048],[-96,0.0509],[-94,0.0545],[-92,0.0562],[-90,0.0694],[-88,0.0759],[-86,0.1023],[-84,0.1388],[-82,0.168],[-80,0.2112],[-78,0.2482],[-76,0.2565],[-74,0.2463],[-72,0.2396],[-70,0.2291],[-68,0.2083],[-66,0.1662],[-64,0.1217],[-62,0.1104],[-60,0.0584],[-58,0.0568],[-56,0.0534],[-54,0.0497],[-52,0.0483],[-48,0.0482],[-46,0.0523],[-44,0.0605],[-42,0.0683],[-40,0.0717],[-38,0.0697],[-36,0.0733],[-34,0.0717],[-32,0.0662],[-30,0.0662],[-28,0.0686],[-26,0.0686],[-24,0.0646],[-22,0.0615],[-20,0.0605],[-16,0.0574],[-12,0.0546],[-10,0.0541],[-8,0.0541],[-6,0.0536],[-4,0.0518],[-2,0.0496],[0,0.05],[2,0.0494],[4,0.0482],[6,0.0486],[8,0.0493],[12,0.052],[14,0.0513],[18,0.0512],[20,0.0506],[22,0.0498],[24,0.0494],[26,0.0493],[28,0.0503],[30,0.051],[32,0.0521],[34,0.0529],[36,0.051],[38,0.0504],[40,0.0517],[42,0.0516],[44,0.0509],[46,0.0509],[48,0.0513],[50,0.0521],[54,0.0504],[56,0.0504],[58,0.0511],[60,0.0513],[62,0.0506],[64,0.0504],[66,0.0505],[68,0.0497],[70,0.0496],[72,0.0493],[74,0.0495],[78,0.0485],[80,0.049],[82,0.0492],[84,0.0479],[86,0.0478],[88,0.0484],[90,0.0485],[92,0.0472],[94,0.0493],[96,0.0547],[98,0.0569],[100,0.06],[102,0.0614],[104,0.0588],[106,0.055],[108,0.0548],[110,0.0553],[112,0.058],[114,0.0586],[116,0.0587],[118,0.0636],[120,0.0648],[122,0.067],[124,0.0663],[126,0.0662],[128,0.0652],[130,0.0652],[132,0.0633],[134,0.0621],[136,0.0637],[138,0.0638],[140,0.0612],[142,0.06],[144,0.0579],[146,0.0577],[148,0.0555],[150,0.0553],[152,0.0543],[154,0.0519],[156,0.0518],[158,0.0511],[160,0.0499],[164,0.0493],[166,0.0475],[168,0.0484],[170,0.0485],[172,0.0466],[174,0.0469],[176,0.0476],[180,0.0471],[184,0.0472],[186,0.047],[188,0.0464],[190,0.0466],[192,0.0473],[194,0.0472],[196,0.0474],[198,0.048],[200,0.0481],[202,0.0494],[204,0.0499],[206,0.0466],[208,0.0476],[210,0.0477],[212,0.0498],[216,0.0496],[218,0.0506],[220,0.0526],[224,0.052],[226,0.0551],[228,0.0549],[230,0.0534],[232,0.0543],[234,0.0542],[236,0.056],[238,0.0574],[240,0.0576],[242,0.0567],[244,0.0552],[246,0.0559],[248,0.0555],[252,0.0513],[254,0.0512],[256,0.0515],[258,0.0501],[260,0.0482],[266,0.0457],[268,0.0468],[270,0.0488],[272,0.0525],[274,0.0536],[276,0.0471],[278,0.0431],[282,0.0431],[284,0.0446],[288,0.0512],[290,0.0528],[292,0.0559],[294,0.0562],[296,0.0547],[298,0.0541],[300,0.0538],[302,0.0526],[304,0.0502],[306,0.0499],[308,0.0512],[310,0.0527],[312,0.0522],[314,0.0486],[316,0.0472],[318,0.0487],[320,0.0491],[322,0.0489],[324,0.0478],[326,0.0486],[328,0.05],[330,0.0486],[332,0.0483],[334,0.0492],[336,0.0535],[338,0.0623],[340,0.0643],[342,0.0673],[344,0.0698],[346,0.0683],[348,0.0744],[350,0.0841],[352,0.0857],[354,0.0855],[356,0.0845],[358,0.085],[360,0.0873],[362,0.0941],[364,0.0956],[366,0.0946],[368,0.0898],[370,0.0856],[374,0.0914],[376,0.0918],[378,0.0909],[380,0.0925],[382,0.0924],[384,0.0913],[386,0.0859],[388,0.0853],[390,0.085],[392,0.093],[394,0.1135],[396,0.1271],[398,0.1314],[400,0.1325],[402,0.1504],[404,0.1422],[406,0.1227],[408,0.128],[410,0.1271],[412,0.1241],[414,0.1251],[416,0.1255],[418,0.1198],[420,0.1131],[422,0.1037],[424,0.1032],[426,0.1032],[428,0.0995],[430,0.098],[432,0.099],[434,0.0991],[436,0.0983],[440,0.0989],[442,0.0982],[444,0.0952],[448,0.0994],[450,0.0975],[452,0.0976],[454,0.0995],[456,0.0993],[458,0.1005],[460,0.1078],[462,0.1142],[464,0.1191],[466,0.1383],[468,0.1529],[470,0.1555],[472,0.1356],[474,0.1162],[476,0.1099],[478,0.1201],[480,0.1201],[482,0.1261],[484,0.1282],[486,0.1426],[488,0.1515],[490,0.1671],[492,0.1851],[494,0.2008],[496,0.2173],[498,0.227],[500,0.2239],[502,0.2252],[504,0.2328],[506,0.2387],[508,0.2432],[510,0.2337],[512,0.232],[514,0.2533],[516,0.2429],[518,0.243],[520,0.2419],[522,0.2351],[524,0.2225],[526,0.2033],[528,0.2088],[530,0.2155],[532,0.216],[534,0.2186],[536,0.2181],[538,0.2344],[540,0.2379],[542,0.2577],[544,0.2621],[546,0.2707],[548,0.2984],[550,0.3127],[552,0.3178],[554,0.316],[556,0.318],[560,0.3174],[562,0.3201],[566,0.3179],[568,0.3179],[570,0.3206],[572,0.3148],[576,0.3117],[578,0.3093],[582,0.303],[584,0.2992],[586,0.2835],[588,0.2729],[590,0.2718],[592,0.2546],[594,0.2498],[596,0.2032],[598,0.1729],[600,0.1335],[602,0.1297],[604,0.0964],[606,0.0922],[608,0.0787],[610,0.0714],[612,0.0685],[614,0.068],[616,0.067],[618,0.0665],[620,0.0667],[622,0.0706],[624,0.0706],[626,0.0679],[628,0.0642],[630,0.0622],[632,0.0626],[636,0.0623],[638,0.0624],[640,0.064],[642,0.0653],[644,0.0647],[646,0.0646],[648,0.0655],[650,0.0656],[652,0.0662],[654,0.0675],[656,0.0675],[658,0.0681],[660,0.0712],[662,0.0727],[664,0.0698],[666,0.0689],[670,0.0693],[672,0.0692],[674,0.0665],[676,0.0668],[678,0.0683],[680,0.0693],[684,0.0697],[686,0.0695],[688,0.0678],[690,0.068],[692,0.0702],[694,0.0719],[696,0.0713],[698,0.0698],[700,0.0699],[702,0.0704],[704,0.0695],[706,0.0714],[708,0.0711],[710,0.0697],[712,0.0695],[714,0.0689],[716,0.0678],[720,0.0678]]
       }
       CTRL_C_jaw_openExtreme {
        type=MovieSceneScriptingFloatChannel
        default=0
        thinned_from=492
        keys [[-262,0.002],[-260,0.0007],[-258,0.0009],[-256,0.0002],[-252,0],[-248,0.0005],[-246,0.0001],[-242,0],[-240,0.0007],[-238,0],[-236,0.0005],[-234,0.0005],[-230,0.001],[-226,0.0006],[-222,0.0007],[-220,0.0002],[-218,0.0002],[-216,0.001],[-214,0.0012],[-212,0.0012],[-210,0.001],[-208,0.0011],[-206,0.0006],[-204,0.001],[-202,0.0007],[-200,0.0003],[-198,0.0007],[-196,0.0009],[-194,0.0009],[-190,0.0012],[-188,0.0008],[-182,0.0006],[-180,0.001],[-178,0.0011],[-174,0.0019],[-172,0.0021],[-170,0.002],[-166,0.0024],[-158,0.0027],[-154,0.0021],[-152,0.0024],[-150,0.002],[-146,0.0024],[-136,0.0024],[-134,0.0029],[-132,0.0025],[-128,0.0026],[-126,0.0025],[-120,0.0026],[-116,0.0022],[-112,0.0023],[-110,0.0021],[-108,0.0021],[-104,0.0024],[-102,0.0021],[-86,0.0021],[-84,0.002],[-76,0.0024],[-74,0.0022],[-72,0.0024],[-70,0.0029],[-68,0.0025],[-66,0.0025],[-64,0.0027],[-62,0.0026],[-60,0.002],[-46,0.0017],[-44,0.0022],[-42,0.002],[-40,0.0022],[-38,0.0022],[-36,0.0024],[-34,0.0024],[-32,0.0022],[-28,0.0022],[-26,0.0024],[-24,0.0022],[-22,0.0024],[-20,0.0024],[-18,0.0022],[-16,0.0024],[-14,0.0022],[0,0.0022],[2,0.0024],[6,0.002],[8,0.0022],[10,0.0022],[12,0.0024],[18,0.0023],[20,0.0024],[28,0.0022],[32,0.0024],[38,0.0022],[40,0.0024],[42,0.0024],[44,0.002],[58,0.0023],[64,0.0022],[66,0.002],[78,0.0022],[80,0.0024],[90,0.0022],[92,0.0024],[94,0.0024],[98,0.002],[100,0.0022],[104,0.0022],[106,0.002],[108,0.0022],[112,0.0019],[116,0.0005],[118,0.001],[122,0.0012],[124,0.0017],[126,0.0017],[128,0.0015],[130,0.0015],[132,0.0017],[134,0.0012],[136,0.0012],[138,0.0015],[140,0.0012],[144,0.0012],[146,0.001],[150,0.0019],[152,0.0015],[154,0.0017],[156,0.0017],[158,0.0015],[160,0.001],[162,0.0015],[164,0.0012],[166,0.0012],[168,0.0007],[170,0.0012],[172,0.001],[174,0.0012],[176,0.0005],[178,0.0007],[182,0.0005],[184,0.0007],[188,0.0006],[190,0.0007],[194,0.0005],[196,0.0006],[198,0.001],[200,0.0011],[204,0.001],[206,0.0007],[208,0.001],[210,0.001],[212,0.0004],[214,0.001],[216,0.0005],[218,0.0007],[220,0.0007],[222,0.001],[224,0.0007],[226,0.0007],[228,0.0005],[230,0.001],[232,0.001],[234,0.0005],[236,0.0007],[238,0.0005],[240,0.001],[242,0.001],[246,0.0005],[248,0.0007],[250,0],[254,0.0005],[256,0],[258,0.0005],[260,0.0005],[262,0],[264,0],[268,0.001],[272,0.0008],[274,0.0022],[276,0.0018],[278,0],[280,0.0014],[282,0.0011],[284,0.0012],[286,0.0016],[290,0.001],[292,0.001],[294,0.0014],[296,0.0007],[298,0.0007],[302,0.0012],[304,0.001],[306,0.0014],[308,0.001],[310,0.0012],[312,0.0005],[314,0.001],[316,0.0007],[318,0.0014],[320,0.0007],[324,0.0014],[328,0.001],[330,0.0012],[332,0.001],[334,0.0015],[336,0.001],[338,0.0017],[340,0.0017],[342,0.0021],[344,0.0018],[346,0.0009],[348,0.0003],[350,0.0012],[356,0.0013],[358,0.0004],[362,0.001],[366,0.0005],[370,0.0007],[372,0.0005],[376,0.0007],[380,0.0003],[384,0.0008],[386,0.002],[388,0.0027],[390,0.0025],[394,0.0013],[396,0.0021],[398,0.0022],[402,0.0018],[404,0.0028],[406,0.0022],[408,0.002],[410,0.0021],[412,0.0017],[416,0.0017],[420,0.0022],[424,0.0021],[426,0.0016],[430,0.0017],[432,0.0019],[438,0.0017],[440,0.0022],[442,0.0017],[444,0.0015],[446,0.0019],[448,0.0017],[458,0.0017],[460,0.0019],[464,0.0015],[466,0.0017],[468,0.0015],[470,0.0003],[472,0.002],[474,0.0021],[476,0.002],[478,0.0009],[482,0.0015],[484,0.0015],[486,0.0005],[488,0.0005],[490,0.0009],[492,0.0005],[494,0.0006],[496,0],[498,0.0007],[500,0],[502,0.0007],[506,0.0002],[508,0.001],[510,0.0007],[512,0],[516,0],[518,0.0002],[520,0.0017],[522,0.0023],[524,0.0019],[526,0.0024],[530,0.0029],[534,0.0024],[536,0.0026],[538,0.0019],[540,0.0017],[542,0.0019],[544,0.0015],[546,0.0007],[548,0.0002],[552,0.0002],[556,0.0004],[558,0.0001],[562,0.0005],[564,0.0003],[566,0.0007],[568,0.0007],[572,0],[576,0.0005],[584,0.0002],[586,0.0005],[590,0.0007],[592,0.0014],[594,0.0014],[596,0.0009],[600,0.0019],[608,0.0017],[614,0.0022],[616,0.0019],[620,0.0019],[622,0.0017],[624,0.0017],[626,0.0015],[628,0.0017],[632,0.0018],[634,0.0015],[636,0.0015],[640,0.0019],[642,0.0019],[644,0.0012],[646,0.0012],[648,0.0019],[650,0.0017],[652,0.0017],[654,0.0015],[656,0.0019],[660,0.0017],[662,0.002],[664,0.002],[666,0.0017],[674,0.0017],[676,0.0019],[678,0.0017],[680,0.0017],[682,0.0015],[684,0.0015],[686,0.0019],[688,0.0015],[690,0.0015],[692,0.0017],[694,0.0015],[696,0.0017],[700,0.0015],[702,0.0019],[704,0.0015],[706,0.0018],[708,0.0017],[712,0.002],[714,0.0017],[716,0.0017],[718,0.0015],[720,0.0015]]
       }
       CTRL_C_jaw_fwdBack {
        type=MovieSceneScriptingFloatChannel
        default=0
        thinned_from=492
        keys [[-262,0.1579],[-260,0.1642],[-258,0.1637],[-256,0.1649],[-254,0.1671],[-252,0.1677],[-250,0.1637],[-248,0.168],[-246,0.1697],[-244,0.1697],[-242,0.1713],[-240,0.1703],[-238,0.1707],[-236,0.1694],[-234,0.1697],[-232,0.1718],[-230,0.1703],[-228,0.1644],[-226,0.1612],[-224,0.161],[-222,0.1598],[-220,0.1605],[-218,0.1617],[-216,0.1643],[-214,0.1613],[-210,0.1602],[-208,0.1681],[-206,0.1744],[-204,0.173],[-202,0.1699],[-200,0.1751],[-198,0.1744],[-196,0.173],[-194,0.1684],[-192,0.1667],[-190,0.1663],[-188,0.1616],[-186,0.162],[-184,0.1645],[-182,0.1628],[-180,0.1642],[-178,0.1695],[-174,0.195],[-172,0.1978],[-170,0.1994],[-168,0.2029],[-166,0.1984],[-164,0.1928],[-162,0.1891],[-160,0.1722],[-158,0.1699],[-156,0.1492],[-154,0.1143],[-152,0.0907],[-150,0.0444],[-148,0.0284],[-146,0.0283],[-144,0.0174],[-142,0.0241],[-140,0.0118],[-138,0.0032],[-136,0.004],[-134,-0.0011],[-132,-0.0044],[-130,-0.0101],[-128,-0.0086],[-126,-0.0107],[-124,-0.0168],[-122,-0.0192],[-120,-0.025],[-118,-0.0252],[-116,-0.0128],[-114,-0.0065],[-112,0.0083],[-110,0.0179],[-108,0.024],[-106,0.0277],[-104,0.0293],[-102,0.0407],[-100,0.0448],[-98,0.0451],[-96,0.0588],[-92,0.0607],[-90,0.0714],[-88,0.0739],[-86,0.0838],[-84,0.0968],[-82,0.1066],[-80,0.1252],[-78,0.1348],[-76,0.1351],[-74,0.0808],[-72,0.0557],[-70,0.0321],[-68,0.0053],[-66,-0.0107],[-64,0.0013],[-62,0.0067],[-60,0.036],[-58,0.0391],[-56,0.0462],[-54,0.0483],[-52,0.0481],[-50,0.0419],[-48,0.0391],[-46,0.0242],[-44,0.0192],[-42,0.0182],[-40,0.0161],[-38,0.0236],[-36,0.037],[-34,0.0442],[-32,0.0556],[-30,0.0631],[-28,0.07],[-26,0.0717],[-24,0.0732],[-22,0.0743],[-20,0.0742],[-18,0.0722],[-16,0.0729],[-14,0.075],[-12,0.0743],[-10,0.0731],[-8,0.0745],[-6,0.0766],[-4,0.0754],[-2,0.0768],[0,0.078],[2,0.0768],[4,0.0772],[6,0.0758],[8,0.0755],[10,0.0764],[12,0.0803],[14,0.0836],[16,0.084],[18,0.0836],[20,0.0813],[22,0.0823],[24,0.0811],[28,0.0797],[32,0.0807],[34,0.0809],[36,0.0806],[38,0.0808],[40,0.0827],[42,0.0855],[44,0.0847],[46,0.0845],[48,0.0834],[50,0.0821],[52,0.0804],[54,0.0792],[56,0.0795],[58,0.0781],[60,0.0788],[62,0.0771],[64,0.0747],[66,0.0771],[68,0.0772],[70,0.077],[72,0.0748],[74,0.075],[76,0.078],[78,0.0788],[80,0.0754],[82,0.0759],[84,0.0767],[86,0.075],[88,0.074],[90,0.0743],[92,0.0655],[94,0.0622],[96,0.0527],[98,0.0483],[100,0.0437],[102,0.0447],[104,0.044],[106,0.0454],[108,0.0465],[110,0.051],[112,0.0557],[114,0.0583],[116,0.0616],[118,0.0656],[120,0.0663],[122,0.0681],[124,0.0707],[126,0.071],[128,0.0746],[130,0.0746],[132,0.0732],[134,0.0723],[136,0.0733],[138,0.0712],[140,0.0701],[142,0.066],[146,0.0663],[148,0.0676],[150,0.0674],[152,0.0687],[154,0.0651],[156,0.065],[158,0.0637],[160,0.0641],[162,0.0658],[164,0.0642],[166,0.0615],[168,0.0606],[172,0.0626],[174,0.0604],[176,0.0614],[178,0.0606],[180,0.061],[182,0.0621],[184,0.062],[186,0.0626],[190,0.066],[192,0.0669],[194,0.067],[196,0.0675],[198,0.0669],[200,0.068],[202,0.068],[204,0.0675],[206,0.0664],[208,0.0685],[210,0.0687],[212,0.0695],[214,0.069],[216,0.0715],[218,0.0728],[220,0.073],[222,0.0727],[224,0.0735],[226,0.0734],[228,0.0743],[230,0.0736],[232,0.0694],[236,0.0674],[238,0.062],[240,0.06],[242,0.0608],[244,0.0612],[246,0.0593],[248,0.0594],[250,0.056],[252,0.0543],[254,0.054],[256,0.0564],[258,0.0555],[260,0.0507],[262,0.0481],[264,0.0487],[266,0.0478],[268,0.05],[270,0.0545],[272,0.0596],[274,0.058],[276,0.0563],[278,0.055],[282,0.0557],[284,0.0594],[286,0.0616],[288,0.0646],[290,0.0645],[292,0.0635],[294,0.0647],[296,0.0629],[298,0.0675],[300,0.0707],[302,0.0709],[304,0.0692],[306,0.0657],[310,0.0682],[312,0.0683],[314,0.0675],[316,0.0664],[318,0.0637],[320,0.0656],[322,0.0661],[326,0.0687],[328,0.0695],[330,0.0691],[332,0.0665],[334,0.0645],[336,0.0722],[338,0.0836],[340,0.0857],[342,0.0852],[344,0.0857],[346,0.0856],[348,0.0902],[350,0.0924],[352,0.0931],[354,0.0931],[356,0.094],[358,0.0978],[360,0.0998],[362,0.1048],[364,0.1059],[366,0.1068],[368,0.1064],[370,0.108],[372,0.111],[374,0.1133],[376,0.1136],[378,0.1129],[380,0.1151],[382,0.1143],[384,0.1114],[386,0.1058],[388,0.1013],[390,0.0985],[392,0.1004],[394,0.1069],[396,0.1074],[398,0.1001],[400,0.0976],[402,0.0925],[404,0.0897],[406,0.0853],[408,0.0814],[410,0.0778],[412,0.0779],[414,0.0802],[416,0.0801],[418,0.0779],[422,0.076],[424,0.0769],[426,0.0782],[428,0.0783],[430,0.0798],[432,0.0816],[434,0.0843],[436,0.0867],[438,0.0889],[440,0.087],[442,0.0865],[444,0.0852],[446,0.0866],[448,0.086],[450,0.0879],[452,0.0875],[454,0.0901],[456,0.0902],[460,0.0958],[464,0.0982],[466,0.1002],[468,0.1008],[470,0.0962],[472,0.0926],[474,0.092],[476,0.0947],[480,0.0957],[482,0.1017],[484,0.1031],[486,0.1108],[488,0.1269],[490,0.1393],[492,0.1599],[494,0.1815],[496,0.193],[498,0.1944],[500,0.1887],[502,0.1846],[504,0.1778],[506,0.1807],[508,0.1759],[512,0.1632],[514,0.1636],[516,0.151],[518,0.1519],[520,0.1625],[522,0.1687],[524,0.1605],[526,0.1419],[528,0.1451],[530,0.1449],[532,0.1466],[534,0.1413],[536,0.1266],[538,0.1285],[540,0.1297],[542,0.1301],[544,0.1356],[546,0.1459],[548,0.1624],[550,0.1704],[552,0.1718],[554,0.1728],[556,0.1751],[558,0.1743],[560,0.1746],[562,0.1756],[564,0.1756],[566,0.1747],[568,0.1747],[570,0.1772],[572,0.1828],[574,0.1812],[576,0.1803],[580,0.1801],[582,0.1772],[584,0.1679],[586,0.1545],[590,0.1447],[592,0.1364],[594,0.1348],[596,0.1265],[598,0.1227],[600,0.112],[602,0.1115],[604,0.1083],[606,0.1058],[608,0.0985],[610,0.0949],[612,0.0925],[614,0.0913],[616,0.0907],[618,0.0919],[620,0.0924],[622,0.0937],[624,0.0937],[626,0.0922],[628,0.0927],[630,0.0885],[632,0.0885],[634,0.0882],[636,0.089],[638,0.089],[640,0.0899],[642,0.0913],[644,0.0911],[646,0.0915],[648,0.0922],[650,0.0915],[652,0.0911],[654,0.092],[656,0.0924],[658,0.0922],[660,0.0942],[662,0.0949],[664,0.0948],[666,0.0943],[668,0.0941],[670,0.0947],[674,0.0916],[676,0.0934],[678,0.0922],[680,0.0925],[682,0.0935],[684,0.0937],[686,0.0941],[690,0.0938],[694,0.097],[696,0.0958],[698,0.095],[700,0.0956],[702,0.0957],[704,0.0966],[706,0.0977],[710,0.0991],[712,0.0992],[714,0.0989],[716,0.1002],[718,0.0994],[720,0.0994]]
       }
       CTRL_R_jaw_ChinRaiseD {
        type=MovieSceneScriptingFloatChannel
        default=0
        thinned_from=492
        keys [[-262,0.0409],[-260,0.0412],[-258,0.0453],[-256,0.045],[-254,0.0467],[-252,0.0471],[-250,0.0508],[-248,0.0466],[-246,0.0417],[-244,0.0407],[-242,0.038],[-240,0.036],[-238,0.0356],[-236,0.0323],[-234,0.0324],[-232,0.033],[-230,0.029],[-228,0.0257],[-226,0.026],[-224,0.0245],[-222,0.0232],[-220,0.0235],[-218,0.0294],[-216,0.0297],[-214,0.0271],[-212,0.0269],[-210,0.0284],[-208,0.0295],[-206,0.0344],[-204,0.0323],[-202,0.0331],[-200,0.0347],[-198,0.0418],[-196,0.0428],[-194,0.041],[-192,0.0442],[-190,0.0452],[-188,0.0409],[-186,0.04],[-184,0.0367],[-182,0.0398],[-180,0.0475],[-178,0.0508],[-176,0.0434],[-174,0.0318],[-172,0.0257],[-170,0.0201],[-168,0.0161],[-166,0.0213],[-164,0.025],[-162,0.0192],[-160,0.0115],[-158,0.0116],[-156,0.013],[-152,0.0161],[-150,0.0383],[-148,0.0447],[-146,0.056],[-144,0.0648],[-142,0.0672],[-140,0.0711],[-138,0.0825],[-136,0.0748],[-134,0.0829],[-132,0.0839],[-130,0.0854],[-128,0.0835],[-126,0.0879],[-124,0.095],[-122,0.1017],[-120,0.1216],[-118,0.1239],[-116,0.124],[-114,0.1189],[-112,0.109],[-110,0.1048],[-108,0.0981],[-106,0.0953],[-104,0.0951],[-102,0.0902],[-100,0.0907],[-98,0.0908],[-96,0.0872],[-92,0.0877],[-90,0.0916],[-88,0.0926],[-86,0.097],[-84,0.0975],[-82,0.0912],[-80,0.0796],[-78,0.064],[-76,0.0604],[-74,0.0554],[-72,0.0615],[-70,0.068],[-68,0.0751],[-66,0.0895],[-64,0.1003],[-62,0.0984],[-60,0.0898],[-58,0.0871],[-56,0.0821],[-54,0.0831],[-52,0.0852],[-50,0.0883],[-48,0.0893],[-46,0.096],[-44,0.0994],[-42,0.0972],[-40,0.0912],[-38,0.0898],[-36,0.0908],[-34,0.0888],[-30,0.0905],[-28,0.0926],[-26,0.0929],[-24,0.0915],[-22,0.0888],[-20,0.0899],[-18,0.0915],[-16,0.0911],[-14,0.0903],[-12,0.091],[-10,0.0932],[-8,0.0942],[-6,0.0944],[-4,0.0941],[-2,0.0907],[0,0.0925],[2,0.0927],[4,0.0914],[6,0.0924],[8,0.0918],[12,0.0922],[14,0.0915],[16,0.0918],[18,0.0925],[20,0.0922],[22,0.0924],[24,0.091],[26,0.0893],[28,0.0896],[30,0.0891],[32,0.0894],[34,0.0888],[36,0.0884],[40,0.0872],[42,0.0875],[44,0.0893],[46,0.0896],[48,0.0903],[50,0.0906],[54,0.0882],[56,0.0882],[58,0.0892],[60,0.0894],[62,0.0882],[64,0.0877],[66,0.087],[70,0.0864],[72,0.085],[74,0.0866],[78,0.0851],[80,0.0862],[82,0.0853],[84,0.0849],[86,0.0859],[88,0.0863],[90,0.0862],[92,0.0879],[94,0.0876],[96,0.0863],[98,0.0861],[100,0.0846],[102,0.0816],[104,0.0783],[106,0.0821],[108,0.0829],[110,0.0833],[112,0.083],[114,0.0823],[116,0.0822],[118,0.0866],[120,0.0877],[122,0.0904],[124,0.0921],[126,0.0924],[128,0.0978],[130,0.0978],[134,0.0958],[136,0.0978],[138,0.0985],[140,0.0975],[142,0.097],[144,0.0979],[148,0.0969],[150,0.0967],[152,0.0952],[154,0.0974],[158,0.0974],[160,0.0948],[162,0.0955],[164,0.0936],[166,0.0934],[168,0.095],[170,0.0934],[172,0.0923],[174,0.0938],[176,0.0917],[178,0.0934],[180,0.0928],[182,0.0929],[186,0.0925],[188,0.0919],[190,0.0917],[192,0.0917],[194,0.0911],[196,0.0916],[198,0.0928],[202,0.0948],[204,0.0949],[206,0.0916],[208,0.0951],[210,0.0953],[212,0.0942],[214,0.0942],[216,0.0936],[218,0.0972],[220,0.0978],[222,0.0951],[224,0.0985],[226,0.103],[228,0.1019],[230,0.1012],[232,0.1026],[234,0.1007],[236,0.1034],[238,0.105],[240,0.1057],[242,0.104],[244,0.1046],[246,0.1061],[248,0.1047],[252,0.1015],[254,0.1039],[256,0.1037],[260,0.1077],[262,0.1067],[264,0.1046],[266,0.1038],[268,0.1089],[270,0.1178],[272,0.1271],[276,0.1402],[278,0.1332],[280,0.133],[282,0.1335],[284,0.1407],[286,0.1447],[288,0.1433],[290,0.143],[292,0.148],[294,0.1468],[296,0.1473],[298,0.1449],[300,0.145],[302,0.1446],[304,0.1407],[306,0.1444],[308,0.144],[310,0.144],[312,0.1425],[314,0.138],[316,0.1365],[318,0.1383],[320,0.1369],[322,0.1363],[324,0.1336],[326,0.1373],[328,0.1343],[330,0.1339],[332,0.1321],[338,0.1224],[340,0.1171],[342,0.1152],[344,0.1127],[346,0.1087],[350,0.1149],[352,0.113],[354,0.1133],[356,0.1142],[358,0.1143],[362,0.1177],[366,0.1162],[368,0.1138],[370,0.1123],[372,0.1124],[374,0.1131],[376,0.1131],[378,0.1098],[380,0.1101],[382,0.1088],[384,0.109],[386,0.1061],[388,0.1085],[390,0.1056],[392,0.0995],[396,0.1033],[398,0.105],[400,0.1059],[402,0.1186],[404,0.1112],[406,0.1195],[408,0.124],[410,0.1245],[412,0.1216],[414,0.1214],[416,0.1199],[418,0.1188],[420,0.1203],[422,0.1188],[424,0.1193],[426,0.1192],[428,0.1172],[430,0.1161],[432,0.1158],[434,0.1141],[436,0.1153],[438,0.1159],[440,0.1149],[442,0.1153],[444,0.1128],[446,0.1134],[448,0.1136],[452,0.1123],[454,0.1121],[456,0.1098],[458,0.1099],[460,0.1136],[462,0.1145],[464,0.1143],[466,0.1189],[468,0.1205],[470,0.1144],[472,0.0958],[474,0.086],[476,0.0824],[478,0.0814],[480,0.0775],[482,0.0731],[484,0.0725],[486,0.073],[488,0.0741],[490,0.0732],[492,0.0577],[494,0.0336],[496,0.0145],[498,0.007],[500,0.0055],[502,0.0077],[504,0.0279],[506,0.0377],[508,0.0444],[510,0.043],[512,0.048],[514,0.0368],[516,0.0246],[518,0.0244],[520,0.0238],[522,0.0288],[524,0.035],[526,0.0589],[530,0.0645],[532,0.0627],[534,0.0614],[536,0.0635],[538,0.0641],[540,0.0641],[542,0.0654],[544,0.0664],[546,0.0595],[548,0.0537],[550,0.0513],[552,0.0501],[554,0.0482],[556,0.0468],[558,0.0457],[560,0.0461],[562,0.0472],[564,0.045],[566,0.0481],[568,0.0484],[570,0.0443],[572,0.0441],[574,0.0473],[576,0.0511],[578,0.0515],[580,0.0499],[582,0.0494],[584,0.0459],[586,0.0398],[588,0.0383],[590,0.0458],[592,0.0523],[594,0.0537],[596,0.0675],[598,0.0785],[600,0.093],[602,0.0937],[604,0.0974],[606,0.0976],[608,0.0975],[610,0.0957],[612,0.0951],[614,0.0953],[616,0.0948],[620,0.0957],[622,0.0979],[626,0.0981],[628,0.0959],[630,0.0968],[634,0.098],[636,0.0987],[638,0.0992],[642,0.0993],[644,0.0983],[646,0.098],[648,0.0994],[652,0.0998],[654,0.0992],[656,0.1005],[658,0.1008],[660,0.1025],[662,0.1032],[664,0.1019],[666,0.1016],[668,0.1008],[670,0.102],[672,0.1013],[674,0.1],[676,0.0988],[678,0.1002],[680,0.1002],[682,0.1007],[684,0.1007],[686,0.0989],[688,0.0992],[690,0.0985],[694,0.1013],[696,0.1],[698,0.1006],[700,0.1006],[702,0.1018],[704,0.1012],[706,0.1027],[708,0.102],[710,0.101],[712,0.1008],[714,0.1002],[716,0.1012],[718,0.1039],[720,0.1039]]
       }
       CTRL_R_jaw_ChinRaiseU {type=MovieSceneScriptingFloatChannel, default=0, const [0,492]}
       CTRL_L_jaw_ChinRaiseU {type=MovieSceneScriptingFloatChannel, default=0, const [0,492]}
       CTRL_L_jaw_ChinRaiseD {
        type=MovieSceneScriptingFloatChannel
        default=0
        thinned_from=492
        keys [[-262,0.0389],[-260,0.0388],[-258,0.043],[-256,0.0426],[-254,0.0442],[-252,0.0444],[-250,0.0479],[-248,0.0439],[-246,0.0391],[-244,0.0382],[-242,0.0355],[-240,0.0336],[-238,0.033],[-236,0.0298],[-234,0.0299],[-232,0.0305],[-230,0.0268],[-228,0.0236],[-226,0.0239],[-224,0.0225],[-222,0.0212],[-220,0.0213],[-218,0.0269],[-216,0.0274],[-214,0.0249],[-212,0.0247],[-210,0.0262],[-208,0.0274],[-206,0.032],[-204,0.03],[-202,0.0307],[-200,0.0323],[-198,0.039],[-196,0.04],[-194,0.0383],[-192,0.0413],[-190,0.0422],[-188,0.0382],[-186,0.0373],[-184,0.0344],[-182,0.0372],[-180,0.0446],[-178,0.0481],[-176,0.0409],[-174,0.0298],[-172,0.0241],[-170,0.0189],[-168,0.0151],[-166,0.0203],[-164,0.0239],[-162,0.0187],[-160,0.0118],[-158,0.012],[-156,0.0139],[-154,0.0163],[-152,0.018],[-150,0.0394],[-148,0.0456],[-146,0.0566],[-144,0.0649],[-142,0.0673],[-140,0.0713],[-138,0.082],[-136,0.0749],[-134,0.0826],[-132,0.0834],[-130,0.0846],[-128,0.0829],[-126,0.087],[-122,0.1009],[-120,0.1207],[-118,0.1239],[-116,0.1254],[-114,0.1205],[-112,0.1111],[-110,0.1073],[-108,0.1009],[-106,0.0983],[-104,0.0982],[-102,0.0933],[-98,0.0936],[-96,0.0899],[-94,0.0896],[-92,0.0896],[-90,0.093],[-88,0.0939],[-86,0.0976],[-84,0.0977],[-82,0.0916],[-80,0.0801],[-78,0.0646],[-76,0.061],[-74,0.0557],[-72,0.0616],[-70,0.0679],[-68,0.0745],[-66,0.0888],[-64,0.1002],[-62,0.0987],[-60,0.0918],[-58,0.0893],[-56,0.0847],[-54,0.0859],[-52,0.0882],[-50,0.0916],[-48,0.0928],[-46,0.0994],[-44,0.103],[-42,0.1005],[-40,0.0944],[-38,0.0925],[-36,0.0933],[-34,0.0911],[-30,0.0926],[-28,0.0945],[-26,0.0949],[-24,0.0936],[-22,0.091],[-20,0.0919],[-18,0.0938],[-16,0.0935],[-14,0.0926],[-12,0.0935],[-10,0.0959],[-8,0.0968],[-4,0.0971],[-2,0.0936],[0,0.0951],[2,0.0953],[4,0.0942],[6,0.0951],[8,0.0947],[10,0.0945],[12,0.0946],[14,0.0939],[16,0.0942],[18,0.0949],[20,0.0946],[22,0.0946],[26,0.0916],[28,0.0918],[30,0.0914],[32,0.092],[34,0.0913],[36,0.091],[38,0.0903],[40,0.0898],[42,0.09],[44,0.0916],[46,0.0919],[48,0.0927],[50,0.0932],[54,0.0909],[56,0.0908],[58,0.0918],[60,0.0919],[62,0.0908],[64,0.0903],[66,0.0893],[68,0.0888],[70,0.0887],[72,0.0878],[74,0.0892],[78,0.0875],[80,0.0886],[82,0.0879],[84,0.0876],[86,0.0886],[88,0.0891],[90,0.0892],[92,0.0907],[94,0.0905],[96,0.0893],[98,0.0887],[100,0.0872],[104,0.0805],[106,0.0844],[108,0.0853],[110,0.0857],[112,0.0852],[114,0.0844],[116,0.084],[118,0.0883],[120,0.0894],[122,0.0923],[124,0.0939],[126,0.0942],[128,0.0995],[130,0.0995],[132,0.0987],[134,0.0976],[136,0.0997],[138,0.1005],[140,0.0994],[142,0.0992],[144,0.1001],[146,0.0994],[148,0.0988],[150,0.0988],[152,0.0974],[154,0.0998],[158,0.1],[160,0.0976],[162,0.0983],[164,0.0968],[166,0.0969],[168,0.0981],[170,0.0966],[172,0.0957],[174,0.0968],[176,0.0948],[178,0.0962],[180,0.0955],[184,0.0957],[186,0.0955],[188,0.0949],[190,0.0946],[192,0.0946],[194,0.094],[196,0.0942],[198,0.0956],[202,0.0975],[204,0.0976],[206,0.0947],[208,0.0981],[210,0.0983],[212,0.0969],[214,0.0969],[216,0.0963],[218,0.0999],[220,0.1003],[222,0.0974],[224,0.1009],[226,0.1051],[230,0.1035],[232,0.1046],[234,0.1029],[236,0.1054],[238,0.107],[240,0.1079],[242,0.1064],[244,0.1069],[246,0.1083],[248,0.1071],[252,0.1041],[254,0.1064],[256,0.1063],[260,0.1106],[262,0.1095],[264,0.1073],[266,0.1069],[268,0.112],[272,0.1302],[276,0.1443],[278,0.1369],[282,0.1375],[284,0.1446],[286,0.1484],[288,0.147],[290,0.1464],[292,0.1515],[294,0.1504],[296,0.1508],[298,0.1484],[302,0.1482],[304,0.1443],[306,0.148],[310,0.1475],[312,0.1461],[314,0.1415],[316,0.1401],[318,0.142],[320,0.1404],[322,0.14],[324,0.1371],[326,0.141],[328,0.1382],[330,0.1377],[332,0.136],[334,0.1325],[336,0.1288],[338,0.1253],[340,0.1199],[342,0.1176],[344,0.115],[346,0.1107],[348,0.1134],[350,0.1164],[352,0.1148],[354,0.1151],[356,0.1158],[358,0.1155],[362,0.1187],[364,0.1182],[366,0.1174],[368,0.1152],[370,0.1136],[372,0.1137],[374,0.1142],[376,0.1142],[378,0.1111],[380,0.1113],[382,0.1101],[384,0.11],[386,0.1073],[388,0.1098],[390,0.1071],[392,0.1006],[394,0.1019],[398,0.1056],[400,0.1065],[402,0.119],[404,0.1119],[406,0.1197],[408,0.1242],[410,0.1243],[412,0.1217],[414,0.1215],[416,0.1198],[418,0.1191],[420,0.1209],[422,0.1197],[424,0.12],[426,0.12],[428,0.1178],[430,0.1168],[432,0.1166],[434,0.1152],[436,0.1165],[438,0.1168],[440,0.116],[442,0.1165],[444,0.114],[446,0.1147],[448,0.1148],[450,0.114],[454,0.1134],[456,0.1108],[458,0.1109],[460,0.1142],[462,0.1153],[464,0.1151],[466,0.1193],[468,0.1205],[470,0.1138],[472,0.0955],[474,0.086],[476,0.0826],[478,0.0813],[480,0.0774],[482,0.073],[484,0.0724],[486,0.0727],[488,0.0738],[490,0.0726],[492,0.0569],[494,0.0329],[496,0.0137],[498,0.0064],[500,0.0049],[502,0.0072],[504,0.0263],[506,0.0357],[508,0.0423],[510,0.041],[512,0.0456],[514,0.0348],[516,0.0227],[520,0.0223],[522,0.0274],[524,0.033],[526,0.0558],[528,0.0585],[530,0.0616],[532,0.0599],[534,0.0586],[536,0.0606],[538,0.0614],[540,0.0613],[544,0.0638],[546,0.0572],[548,0.0513],[550,0.0489],[552,0.0477],[554,0.0458],[558,0.0431],[560,0.0437],[562,0.0446],[564,0.0424],[566,0.0453],[568,0.0456],[570,0.0418],[572,0.0413],[574,0.0445],[576,0.0483],[578,0.0487],[580,0.0471],[582,0.0466],[584,0.0433],[586,0.0371],[588,0.0356],[590,0.0431],[592,0.0498],[594,0.0512],[596,0.0655],[598,0.077],[600,0.0921],[602,0.0929],[604,0.0975],[608,0.0982],[610,0.0967],[612,0.0964],[614,0.0968],[616,0.0962],[620,0.0971],[622,0.099],[626,0.0995],[628,0.0972],[630,0.0981],[634,0.0994],[636,0.1004],[638,0.1008],[642,0.1006],[644,0.0995],[646,0.0994],[648,0.1009],[650,0.1013],[654,0.1008],[656,0.1022],[658,0.1026],[660,0.1039],[662,0.1046],[664,0.1033],[666,0.1029],[668,0.1022],[670,0.1035],[672,0.1028],[674,0.1014],[676,0.1003],[678,0.1015],[680,0.1013],[682,0.102],[684,0.102],[686,0.1002],[688,0.1005],[690,0.0999],[694,0.1027],[696,0.1015],[698,0.1019],[700,0.1017],[702,0.103],[704,0.1024],[706,0.1037],[708,0.1033],[710,0.1024],[712,0.1021],[714,0.1014],[716,0.1026],[718,0.1054],[720,0.1054]]
       }
       CTRL_R_mouth_lipsBlow {type=MovieSceneScriptingFloatChannel, default=0, const [0,492]}
       CTRL_L_mouth_lipsTogetherU {
        type=MovieSceneScriptingFloatChannel
        default=0
        thinned_from=492
        keys [[-262,0.4635],[-260,0.4034],[-258,0.3622],[-254,0.3469],[-252,0.3536],[-250,0.3537],[-248,0.3378],[-244,0.3408],[-242,0.3405],[-240,0.3493],[-238,0.3358],[-236,0.3403],[-234,0.336],[-232,0.3048],[-230,0.2962],[-228,0.2987],[-226,0.2916],[-224,0.2801],[-222,0.272],[-220,0.2801],[-218,0.2568],[-216,0.2543],[-214,0.2478],[-212,0.2462],[-210,0.2428],[-208,0.2534],[-206,0.2521],[-204,0.2527],[-202,0.2715],[-200,0.2701],[-198,0.2745],[-196,0.2714],[-194,0.2692],[-192,0.2729],[-190,0.2757],[-188,0.3162],[-186,0.3181],[-184,0.3232],[-182,0.3404],[-180,0.3755],[-178,0.3939],[-176,0.4477],[-174,0.5355],[-170,0.6326],[-168,0.6847],[-166,0.7058],[-164,0.7621],[-162,0.8139],[-160,0.8619],[-158,0.8638],[-156,0.8816],[-154,0.9266],[-152,0.9491],[-150,0.9051],[-148,0.8761],[-146,0.8696],[-144,0.8455],[-142,0.8349],[-140,0.8261],[-138,0.8244],[-136,0.8393],[-134,0.8081],[-132,0.8294],[-130,0.8052],[-128,0.8217],[-126,0.8165],[-124,0.7991],[-122,0.7595],[-120,0.6138],[-118,0.3091],[-116,0.0375],[-114,0.0395],[-112,0.0424],[-110,0.0438],[-108,0.0423],[-106,0.0449],[-104,0.0488],[-102,0.0433],[-100,0.0446],[-98,0.045],[-96,0.0423],[-94,0.0382],[-92,0.0461],[-90,0.1926],[-88,0.2566],[-86,0.4842],[-84,0.6556],[-82,0.7204],[-78,0.804],[-76,0.8159],[-74,0.8779],[-72,0.8551],[-70,0.8155],[-68,0.834],[-66,0.7355],[-64,0.439],[-62,0.3461],[-60,0.0419],[-58,0.0357],[-56,0.0361],[-54,0.0419],[-52,0.0383],[-50,0.0372],[-48,0.0363],[-46,0.0294],[-44,0.0363],[-42,0.0838],[-40,0.0351],[-38,0.0313],[-36,0.1055],[-34,0.0523],[-32,0.0487],[-30,0.0515],[-28,0.1018],[-26,0.0588],[-24,0.0347],[-22,0.0363],[-20,0.0396],[-18,0.0391],[-16,0.0456],[-14,0.0398],[-12,0.0409],[-10,0.04],[-8,0.0365],[-6,0.0409],[-4,0.0392],[-2,0.0394],[0,0.0432],[2,0.0415],[4,0.045],[6,0.0358],[8,0.0489],[10,0.043],[12,0.0426],[14,0.0479],[16,0.0465],[18,0.0444],[20,0.049],[22,0.0472],[24,0.0465],[26,0.0492],[28,0.0452],[30,0.049],[32,0.0475],[34,0.0462],[36,0.0483],[38,0.0428],[40,0.0425],[42,0.0471],[44,0.0388],[46,0.0395],[48,0.0422],[50,0.0422],[52,0.0432],[54,0.0436],[56,0.0432],[58,0.0454],[60,0.0444],[64,0.0434],[66,0.0401],[68,0.049],[70,0.0479],[72,0.0414],[74,0.0468],[76,0.0453],[80,0.0476],[82,0.0452],[84,0.0486],[86,0.0479],[88,0.0456],[90,0.0459],[92,0.0526],[94,0.0417],[96,0.0401],[98,0.0416],[100,0.0354],[102,0.0463],[104,0.0397],[106,0.0388],[108,0.0402],[110,0.0539],[112,0.0803],[114,0.0361],[116,0.042],[118,0.1224],[120,0.1392],[122,0.0985],[124,0.125],[126,0.1215],[128,0.0654],[130,0.0654],[132,0.0728],[134,0.0532],[136,0.1199],[138,0.0575],[140,0.0353],[142,0.0212],[144,0.0167],[146,0.0111],[148,0.0163],[150,0.0285],[152,0.0197],[154,0.024],[156,0.0234],[158,0.0142],[160,0.015],[162,0.019],[164,0.0056],[166,0.0101],[168,0.0051],[170,0.0153],[172,0.0159],[174,0.0145],[176,0.0002],[178,0.0042],[180,0.0052],[184,0.0052],[186,0.0037],[188,0.0014],[190,0.0034],[192,0],[194,0],[196,0.0026],[198,0.0128],[200,0.0143],[202,0.0066],[204,0.0047],[206,0],[210,0],[212,0.0013],[214,0.0049],[216,0.0003],[218,0.0047],[220,0.0046],[222,0.0096],[224,0.0093],[226,0.0083],[228,0.0003],[230,0.0179],[234,0.0003],[236,0.0082],[238,0.0013],[240,0.0079],[242,0.0132],[246,0.0042],[248,0],[252,0],[254,0.0045],[256,0],[266,0],[268,0.015],[270,0.015],[272,0.0183],[274,0.033],[276,0.0418],[278,0.0002],[280,0.0318],[282,0.0196],[284,0.0258],[286,0.031],[290,0.0176],[292,0.0093],[294,0.0216],[296,0.0135],[298,0.0093],[300,0.0136],[302,0.0234],[304,0.02],[306,0.0237],[308,0.0189],[310,0.0222],[312,0.0097],[314,0.0154],[316,0.0156],[318,0.029],[320,0.0109],[322,0.0244],[324,0.0261],[326,0.0236],[328,0.0125],[330,0.0203],[332,0.0252],[334,0.0294],[336,0.0142],[338,0.0305],[340,0.0317],[342,0.0378],[344,0.0339],[346,0.0511],[348,0.1907],[352,0.2942],[354,0.2835],[356,0.2535],[358,0.2875],[360,0.3092],[362,0.3235],[364,0.324],[366,0.3139],[368,0.2779],[370,0.2875],[372,0.308],[376,0.3089],[378,0.3309],[380,0.3505],[382,0.3799],[384,0.4289],[386,0.3458],[388,0.4363],[390,0.4035],[392,0.52],[394,0.6683],[396,0.733],[398,0.7625],[400,0.7655],[402,0.6828],[404,0.7112],[406,0.6772],[408,0.6735],[410,0.6994],[412,0.683],[414,0.6842],[416,0.6509],[418,0.6257],[420,0.5368],[422,0.5134],[424,0.5348],[426,0.5369],[428,0.5001],[430,0.5176],[432,0.5194],[434,0.4917],[436,0.4995],[438,0.4785],[440,0.485],[442,0.4725],[444,0.4338],[446,0.4752],[448,0.4563],[450,0.4825],[452,0.4753],[454,0.4783],[456,0.4938],[458,0.5124],[460,0.5416],[462,0.5508],[464,0.5512],[466,0.6332],[468,0.656],[470,0.6326],[472,0.6427],[474,0.6147],[476,0.6044],[478,0.6517],[480,0.5921],[482,0.6739],[484,0.6977],[486,0.7173],[488,0.7101],[490,0.7175],[492,0.7896],[494,0.781],[496,0.6886],[498,0.6506],[500,0.5907],[502,0.5722],[504,0.5133],[506,0.505],[508,0.5106],[510,0.5287],[512,0.4754],[514,0.5068],[516,0.5819],[518,0.5736],[520,0.5205],[522,0.5325],[524,0.4385],[526,0.2669],[528,0.2909],[530,0.3191],[532,0.3111],[534,0.2715],[536,0.2589],[538,0.2886],[540,0.2944],[542,0.3202],[544,0.3581],[546,0.3948],[548,0.3955],[550,0.3984],[554,0.3775],[556,0.3696],[558,0.3694],[560,0.3536],[562,0.353],[564,0.371],[566,0.3598],[568,0.3579],[570,0.3621],[572,0.3714],[576,0.3777],[578,0.3865],[580,0.3899],[582,0.3635],[584,0.3335],[586,0.3115],[588,0.2864],[590,0.3462],[592,0.4492],[594,0.4703],[596,0.6158],[598,0.6485],[600,0.6024],[602,0.5811],[604,0.415],[606,0.3763],[608,0.2528],[610,0.196],[612,0.1661],[614,0.1586],[616,0.129],[618,0.1344],[620,0.146],[622,0.1349],[624,0.115],[626,0.0839],[628,0.0439],[630,0.0472],[632,0.0442],[634,0.0273],[636,0.0427],[638,0.0394],[640,0.0802],[642,0.0744],[644,0.0611],[646,0.0794],[648,0.0737],[650,0.0433],[652,0.0966],[654,0.087],[656,0.0874],[658,0.1099],[660,0.1699],[662,0.1486],[664,0.087],[666,0.1202],[668,0.0948],[670,0.1253],[672,0.0757],[674,0.0495],[676,0.0895],[678,0.0997],[680,0.1175],[682,0.099],[684,0.1179],[686,0.1143],[688,0.0815],[690,0.1194],[692,0.1313],[694,0.1533],[696,0.144],[698,0.117],[700,0.1313],[702,0.1179],[704,0.1075],[706,0.1386],[708,0.1213],[710,0.1132],[712,0.1127],[714,0.1078],[716,0.0741],[718,0.0946],[720,0.0946]]
       }
       CTRL_L_mouth_lipsTogetherD {
        type=MovieSceneScriptingFloatChannel
        default=0
        thinned_from=492
        keys [[-262,0.1906],[-260,0.2114],[-258,0.2128],[-256,0.2076],[-254,0.197],[-252,0.2008],[-250,0.195],[-248,0.1863],[-246,0.1834],[-244,0.183],[-242,0.1785],[-240,0.1797],[-238,0.1784],[-236,0.1829],[-234,0.1841],[-232,0.1887],[-230,0.1775],[-228,0.1646],[-226,0.1511],[-224,0.1357],[-222,0.1242],[-220,0.1336],[-218,0.1518],[-216,0.1567],[-214,0.1379],[-212,0.1355],[-210,0.1467],[-208,0.1644],[-206,0.1799],[-204,0.1654],[-202,0.1532],[-200,0.1687],[-198,0.1806],[-196,0.1708],[-194,0.1653],[-192,0.1739],[-190,0.1768],[-188,0.1727],[-186,0.1731],[-184,0.1859],[-182,0.202],[-178,0.2603],[-176,0.2771],[-174,0.2868],[-172,0.3015],[-170,0.3113],[-168,0.3246],[-166,0.3364],[-164,0.334],[-162,0.3094],[-160,0.2274],[-158,0.2177],[-156,0.1322],[-154,0.0383],[-152,0.0095],[-150,0.0079],[-148,0.0087],[-146,0.011],[-142,0.011],[-140,0.0108],[-138,0.0111],[-136,0.0111],[-134,0.0123],[-130,0.0116],[-128,0.0126],[-126,0.0119],[-124,0.0138],[-122,0.0153],[-120,0.0217],[-118,0.0341],[-116,0.0458],[-114,0.0483],[-112,0.0522],[-110,0.0491],[-108,0.0487],[-106,0.0493],[-102,0.0484],[-100,0.0541],[-98,0.055],[-96,0.0515],[-94,0.0425],[-92,0.0406],[-90,0.0313],[-88,0.0422],[-86,0.0792],[-82,0.1539],[-80,0.1742],[-78,0.136],[-76,0.1151],[-74,0.0107],[-72,0.0095],[-70,0.0108],[-68,0.0128],[-66,0.0174],[-64,0.0326],[-62,0.0347],[-60,0.0438],[-58,0.0402],[-56,0.0408],[-54,0.0467],[-52,0.0401],[-50,0.041],[-48,0.0415],[-46,0.0336],[-44,0.0389],[-42,0.0305],[-40,0.0387],[-38,0.0382],[-36,0.0381],[-34,0.0404],[-32,0.041],[-30,0.0404],[-28,0.035],[-26,0.0404],[-22,0.0443],[-20,0.0411],[-16,0.054],[-14,0.0486],[-12,0.0499],[-10,0.0489],[-8,0.0411],[-6,0.0454],[-4,0.0441],[-2,0.049],[0,0.0481],[2,0.0506],[4,0.055],[6,0.0549],[8,0.0539],[10,0.052],[12,0.0472],[14,0.048],[16,0.0479],[18,0.0491],[20,0.0539],[22,0.0521],[24,0.0515],[26,0.0541],[28,0.0524],[30,0.0525],[32,0.0476],[34,0.0499],[36,0.0538],[38,0.0564],[40,0.0476],[42,0.0518],[44,0.0436],[46,0.0431],[48,0.0435],[52,0.0501],[54,0.051],[56,0.0481],[60,0.0526],[62,0.0501],[64,0.0523],[66,0.0407],[68,0.049],[70,0.0491],[72,0.0498],[74,0.0472],[76,0.0471],[78,0.0515],[80,0.0538],[82,0.0514],[84,0.0508],[86,0.0529],[88,0.0507],[90,0.051],[92,0.0579],[94,0.0463],[96,0.04],[98,0.0416],[100,0.0392],[102,0.0371],[104,0.0436],[106,0.0303],[108,0.0378],[110,0.0368],[112,0.0312],[114,0.0158],[116,0.0027],[118,0.0143],[120,0.0178],[122,0.0151],[124,0.0184],[126,0.0187],[128,0.0231],[130,0.0231],[132,0.0268],[134,0.0204],[136,0.0149],[138,0.0235],[140,0.016],[142,0.0207],[144,0.0169],[146,0.0154],[148,0.0206],[150,0.0292],[152,0.0196],[154,0.024],[156,0.0234],[158,0.0142],[160,0.015],[162,0.0191],[164,0.0105],[166,0.0153],[168,0.0147],[170,0.0204],[172,0.0213],[174,0.0195],[176,0.0054],[178,0.0088],[180,0],[182,0.0003],[184,0.0052],[186,0.0053],[188,0.0067],[190,0.0102],[192,0.0083],[194,0.0041],[196,0.0026],[198,0.0153],[200,0.0193],[202,0.0115],[204,0.0096],[206,0.0047],[208,0.0049],[210,0.0049],[212,0.0062],[214,0.0098],[216,0.0003],[218,0.0047],[220,0.0089],[222,0.0096],[224,0.0093],[226,0.0041],[228,0.0001],[230,0.018],[232,0.0135],[234,0.0047],[236,0.0083],[238,0.0049],[240,0.0081],[242,0.0131],[244,0.0088],[246,0.0047],[248,0.0089],[250,0.0004],[252,0],[254,0.0045],[256,0],[262,0],[264,0.0002],[266,0.0054],[268,0.0248],[270,0.02],[272,0.0186],[274,0.0378],[276,0.0523],[278,0.0003],[280,0.0371],[282,0.0252],[284,0.031],[286,0.0323],[288,0.0196],[290,0.0176],[292,0.0097],[294,0.0301],[296,0.0136],[298,0.0138],[300,0.0182],[302,0.0237],[304,0.025],[306,0.0285],[308,0.0236],[310,0.0266],[312,0.0098],[314,0.0204],[316,0.0207],[318,0.0339],[320,0.0156],[322,0.0255],[324,0.0352],[326,0.0238],[328,0.0175],[330,0.0255],[332,0.0202],[334,0.0245],[336,0.0184],[338,0.0323],[340,0.034],[342,0.0363],[344,0.0305],[346,0.0136],[348,0.0091],[350,0.0144],[352,0.017],[354,0.0185],[356,0.021],[358,0.0197],[360,0.0295],[362,0.0649],[364,0.0667],[366,0.0672],[368,0.0421],[370,0.0628],[372,0.1002],[374,0.1062],[376,0.1065],[378,0.1112],[380,0.1119],[382,0.0875],[384,0.0582],[386,0.0347],[388,0.036],[390,0.0333],[392,0.0717],[394,0.1205],[396,0.1102],[398,0.1095],[400,0.1145],[402,0.1934],[404,0.1283],[406,0.121],[408,0.113],[410,0.0952],[412,0.096],[414,0.1238],[416,0.0945],[418,0.0716],[420,0.0564],[422,0.0299],[424,0.0381],[426,0.0305],[428,0.0176],[430,0.0273],[432,0.0357],[434,0.028],[436,0.0371],[438,0.0347],[440,0.0388],[442,0.0256],[444,0.0218],[446,0.0453],[448,0.0323],[450,0.0275],[452,0.0329],[454,0.0409],[456,0.0219],[458,0.052],[460,0.0724],[462,0.0999],[464,0.1221],[466,0.1841],[468,0.2028],[470,0.175],[472,0.0425],[474,0.0243],[476,0.0483],[478,0.059],[480,0.042],[482,0.0848],[484,0.0972],[486,0.1256],[488,0.176],[490,0.2338],[492,0.2569],[494,0.236],[496,0.1598],[498,0.1245],[500,0.1432],[502,0.1641],[504,0.2417],[506,0.2927],[508,0.2816],[510,0.258],[512,0.2717],[514,0.2444],[516,0.131],[518,0.1382],[520,0.2018],[522,0.2049],[524,0.1779],[526,0.2124],[528,0.2297],[530,0.2229],[532,0.2255],[534,0.1955],[536,0.1498],[538,0.2008],[540,0.2133],[542,0.2203],[544,0.2234],[546,0.2306],[548,0.2494],[550,0.2598],[552,0.248],[554,0.2455],[556,0.2421],[558,0.2274],[560,0.2339],[562,0.2368],[564,0.219],[566,0.2336],[568,0.2359],[570,0.2274],[572,0.2539],[574,0.2493],[576,0.2471],[578,0.2417],[580,0.239],[582,0.2352],[584,0.2092],[586,0.1824],[588,0.1809],[590,0.2184],[592,0.2349],[594,0.2355],[596,0.2102],[598,0.1936],[600,0.1293],[602,0.1191],[604,0.0357],[606,0.033],[608,0.0255],[610,0.0248],[614,0.0322],[616,0.026],[618,0.0244],[620,0.0249],[622,0.021],[624,0.021],[626,0.0182],[628,0.0201],[630,0.0238],[632,0.0241],[634,0.0156],[636,0.0237],[638,0.0273],[640,0.0262],[642,0.0222],[644,0.0118],[646,0.0154],[648,0.0254],[650,0.0227],[652,0.0218],[654,0.0184],[656,0.0251],[658,0.018],[660,0.0205],[662,0.0237],[664,0.0254],[666,0.0183],[668,0.0222],[670,0.0239],[672,0.0218],[674,0.026],[676,0.0284],[678,0.0216],[680,0.0207],[682,0.0177],[684,0.021],[686,0.0249],[688,0.0182],[690,0.0211],[692,0.0171],[694,0.0167],[696,0.0173],[698,0.0168],[700,0.014],[702,0.0243],[704,0.0215],[706,0.0236],[708,0.0221],[710,0.0236],[712,0.0246],[714,0.0212],[716,0.0221],[718,0.0176],[720,0.0176]]
       }
       CTRL_R_mouth_lipsTogetherU {
        type=MovieSceneScriptingFloatChannel
        default=0
        thinned_from=492
        keys [[-262,0.4669],[-260,0.4069],[-258,0.3658],[-254,0.3495],[-252,0.3567],[-250,0.3566],[-248,0.3404],[-244,0.343],[-242,0.3429],[-240,0.3517],[-238,0.3389],[-236,0.3417],[-234,0.3376],[-232,0.3075],[-230,0.2994],[-228,0.3014],[-226,0.2938],[-224,0.2823],[-222,0.2738],[-220,0.2815],[-218,0.2583],[-216,0.2552],[-214,0.2492],[-212,0.2477],[-210,0.2444],[-208,0.2549],[-206,0.2538],[-204,0.2555],[-202,0.2737],[-200,0.2709],[-198,0.276],[-194,0.2694],[-192,0.274],[-190,0.2772],[-188,0.3177],[-186,0.3195],[-184,0.3249],[-182,0.3423],[-180,0.3784],[-178,0.3954],[-176,0.45],[-174,0.5387],[-170,0.6375],[-168,0.6897],[-166,0.7099],[-164,0.7669],[-162,0.8192],[-160,0.8687],[-158,0.8706],[-156,0.8884],[-154,0.9324],[-152,0.9545],[-150,0.9115],[-148,0.881],[-146,0.874],[-144,0.8486],[-142,0.8402],[-140,0.8291],[-138,0.8279],[-136,0.8422],[-134,0.8111],[-132,0.8328],[-130,0.8086],[-128,0.8246],[-126,0.8189],[-124,0.8012],[-122,0.761],[-120,0.6155],[-118,0.312],[-116,0.0375],[-114,0.0378],[-112,0.0375],[-110,0.0387],[-106,0.0416],[-104,0.0434],[-102,0.0381],[-100,0.0396],[-98,0.04],[-96,0.0376],[-94,0.034],[-92,0.0415],[-90,0.1957],[-88,0.2605],[-86,0.4897],[-84,0.6598],[-82,0.7261],[-80,0.767],[-78,0.8094],[-76,0.8213],[-74,0.8824],[-72,0.8591],[-70,0.8198],[-68,0.838],[-66,0.7397],[-64,0.4391],[-62,0.3459],[-60,0.0375],[-58,0.0328],[-56,0.0341],[-54,0.0398],[-52,0.0383],[-48,0.0363],[-46,0.0254],[-44,0.0363],[-42,0.084],[-40,0.0387],[-38,0.0382],[-36,0.1086],[-34,0.0524],[-32,0.0525],[-30,0.0588],[-28,0.1018],[-26,0.0588],[-24,0.0386],[-22,0.0403],[-20,0.0403],[-18,0.0386],[-16,0.0419],[-14,0.0398],[-12,0.0409],[-10,0.0401],[-8,0.0365],[-6,0.0409],[-4,0.0392],[-2,0.0392],[0,0.0385],[2,0.0405],[4,0.045],[6,0.04],[8,0.0392],[10,0.0381],[12,0.0471],[14,0.0434],[16,0.043],[18,0.0416],[20,0.0427],[22,0.0452],[24,0.0397],[26,0.0432],[28,0.0424],[30,0.0412],[32,0.0416],[34,0.0415],[36,0.0438],[38,0.0425],[40,0.0425],[42,0.0471],[44,0.0388],[46,0.0395],[48,0.0422],[50,0.0421],[52,0.0432],[54,0.0436],[56,0.0432],[58,0.0425],[62,0.0437],[64,0.0393],[66,0.0395],[68,0.0441],[70,0.0436],[72,0.0399],[74,0.0396],[76,0.0403],[80,0.0407],[84,0.0456],[86,0.0478],[88,0.0456],[90,0.0459],[92,0.0425],[94,0.0417],[96,0.0389],[98,0.0374],[100,0.0315],[102,0.0463],[104,0.0397],[106,0.0388],[108,0.0401],[110,0.0584],[112,0.0836],[114,0.0389],[116,0.0433],[118,0.1224],[120,0.1392],[122,0.0984],[124,0.125],[126,0.1212],[128,0.0615],[130,0.0615],[132,0.0727],[134,0.0493],[136,0.1163],[138,0.0574],[140,0.0314],[142,0.0211],[144,0.0169],[146,0.0143],[148,0.0173],[150,0.0285],[152,0.0153],[154,0.0192],[156,0.0189],[158,0.0141],[160,0.0102],[162,0.0143],[164,0.0054],[166,0.01],[168,0.0003],[170,0.0053],[172,0.0105],[174,0.0144],[176,0.0002],[178,0.0039],[180,0],[186,0],[188,0.0014],[190,0.0034],[192,0],[194,0],[196,0.0026],[198,0.0128],[200,0.0171],[202,0.0081],[204,0.0047],[206,0],[208,0.0046],[210,0.0049],[212,0.0026],[214,0.0098],[216,0.0003],[218,0.0047],[220,0.0002],[222,0.0048],[224,0.0046],[226,0],[228,0],[230,0.0134],[232,0.0047],[234,0.0003],[236,0.0082],[238,0.0051],[240,0.0121],[242,0.0089],[244,0.0088],[246,0.0006],[248,0.0044],[250,0.0002],[252,0],[254,0.0045],[256,0],[262,0],[264,0.0002],[266,0.0054],[268,0.0152],[270,0.02],[272,0.0142],[274,0.033],[276,0.0366],[278,0.0002],[280,0.0318],[282,0.0224],[284,0.0259],[286,0.0343],[288,0.0197],[290,0.0134],[292,0.0133],[294,0.0258],[296,0.0136],[298,0.0095],[300,0.0181],[302,0.0231],[304,0.015],[306,0.0189],[308,0.0142],[310,0.022],[312,0.0097],[314,0.0154],[316,0.0155],[318,0.0242],[320,0.0103],[322,0.0162],[324,0.03],[326,0.019],[328,0.0163],[330,0.0204],[332,0.0205],[334,0.0294],[336,0.0143],[338,0.0323],[340,0.0295],[342,0.0393],[344,0.0374],[346,0.0517],[348,0.1959],[350,0.2481],[352,0.2972],[354,0.2878],[356,0.2614],[358,0.2931],[360,0.3143],[362,0.3284],[364,0.3305],[366,0.3208],[368,0.2843],[370,0.2954],[372,0.3161],[374,0.3144],[376,0.3143],[378,0.3388],[380,0.3553],[382,0.3824],[384,0.4305],[386,0.3437],[388,0.4362],[390,0.4021],[392,0.521],[394,0.6728],[396,0.7382],[398,0.7675],[400,0.7704],[402,0.6889],[404,0.7158],[406,0.6844],[408,0.677],[410,0.7053],[412,0.6893],[414,0.692],[416,0.6563],[418,0.6301],[420,0.5412],[422,0.5159],[424,0.5396],[426,0.5426],[428,0.5074],[430,0.523],[432,0.5243],[434,0.4991],[436,0.5067],[438,0.4787],[440,0.4874],[442,0.4751],[444,0.4365],[446,0.48],[448,0.4612],[450,0.4874],[452,0.48],[454,0.4807],[456,0.4985],[458,0.5144],[460,0.5482],[462,0.5571],[464,0.5557],[466,0.6364],[468,0.6608],[470,0.6342],[472,0.6464],[474,0.6194],[476,0.6097],[478,0.6551],[480,0.5986],[482,0.6814],[484,0.7051],[486,0.7241],[488,0.7165],[490,0.7234],[492,0.7974],[494,0.7859],[496,0.6948],[498,0.6541],[500,0.5937],[502,0.5751],[504,0.5163],[506,0.5094],[508,0.5145],[510,0.5321],[512,0.4786],[514,0.5108],[516,0.5839],[518,0.5758],[520,0.5246],[522,0.5373],[524,0.4419],[526,0.2691],[528,0.2935],[530,0.323],[532,0.3156],[534,0.2744],[536,0.2612],[538,0.2906],[540,0.2963],[542,0.3238],[544,0.3612],[546,0.3966],[550,0.4014],[554,0.3795],[556,0.3723],[558,0.3718],[560,0.3562],[562,0.3554],[564,0.3734],[566,0.3629],[568,0.361],[570,0.3645],[572,0.3739],[576,0.3808],[578,0.3897],[580,0.3932],[582,0.366],[584,0.3349],[586,0.3136],[588,0.2884],[590,0.3493],[592,0.4522],[594,0.4733],[596,0.6196],[598,0.6534],[600,0.6063],[602,0.5852],[604,0.4175],[606,0.3789],[608,0.2559],[610,0.1983],[612,0.1678],[614,0.1621],[616,0.1302],[618,0.137],[620,0.1497],[622,0.1383],[624,0.1185],[626,0.0875],[628,0.0515],[630,0.0512],[632,0.0481],[634,0.0312],[636,0.0427],[638,0.0394],[640,0.0837],[642,0.0707],[644,0.0537],[646,0.0757],[648,0.0736],[650,0.0471],[652,0.1002],[654,0.0906],[656,0.091],[658,0.1133],[660,0.1676],[662,0.152],[664,0.0906],[666,0.1238],[668,0.1013],[670,0.1288],[672,0.0758],[674,0.0531],[676,0.0896],[678,0.1032],[680,0.1243],[682,0.1026],[684,0.1213],[686,0.1144],[688,0.0817],[690,0.1229],[692,0.1347],[694,0.1566],[696,0.1478],[698,0.124],[700,0.1382],[702,0.1214],[704,0.1111],[706,0.1403],[708,0.1214],[710,0.1132],[712,0.1127],[714,0.1112],[716,0.0741],[718,0.098],[720,0.098]]
       }
       CTRL_L_mouth_lipsBlow {type=MovieSceneScriptingFloatChannel, default=0, const [0,492]}
       CTRL_R_mouth_lipsTogetherD {
        type=MovieSceneScriptingFloatChannel
        default=0
        thinned_from=492
        keys [[-262,0.1887],[-260,0.2102],[-258,0.2099],[-256,0.2054],[-254,0.1957],[-252,0.1991],[-250,0.1924],[-248,0.1844],[-246,0.1811],[-244,0.1806],[-242,0.1769],[-240,0.1781],[-238,0.1768],[-236,0.1806],[-234,0.182],[-232,0.1878],[-230,0.1758],[-226,0.1497],[-224,0.1346],[-222,0.1235],[-220,0.1329],[-218,0.1506],[-216,0.1553],[-214,0.1365],[-212,0.134],[-210,0.1446],[-208,0.1629],[-206,0.1785],[-204,0.1639],[-202,0.1517],[-200,0.1686],[-198,0.1791],[-196,0.1686],[-194,0.1635],[-192,0.173],[-190,0.176],[-188,0.1719],[-186,0.1723],[-184,0.1842],[-182,0.2003],[-178,0.2581],[-176,0.2752],[-174,0.2846],[-172,0.2987],[-170,0.3087],[-168,0.3215],[-166,0.333],[-164,0.3295],[-162,0.3059],[-160,0.2252],[-158,0.2157],[-156,0.1315],[-154,0.0374],[-152,0.0087],[-150,0.0089],[-148,0.0087],[-146,0.011],[-144,0.011],[-142,0.0099],[-140,0.0107],[-138,0.01],[-136,0.01],[-134,0.0123],[-132,0.0108],[-130,0.0114],[-126,0.0119],[-124,0.0138],[-122,0.0149],[-120,0.0196],[-118,0.0295],[-116,0.0333],[-114,0.0353],[-112,0.0412],[-110,0.0437],[-108,0.0433],[-106,0.047],[-104,0.0489],[-102,0.0484],[-100,0.0496],[-98,0.05],[-96,0.0424],[-94,0.0424],[-92,0.0413],[-90,0.0313],[-88,0.0417],[-86,0.0775],[-84,0.1161],[-82,0.1526],[-80,0.1736],[-78,0.1354],[-76,0.1142],[-74,0.0099],[-72,0.0094],[-70,0.0108],[-68,0.0115],[-66,0.0172],[-64,0.0276],[-62,0.0289],[-60,0.035],[-58,0.0329],[-56,0.0361],[-54,0.0419],[-52,0.0383],[-48,0.0363],[-46,0.0254],[-44,0.035],[-42,0.0272],[-40,0.0352],[-38,0.0347],[-36,0.0317],[-34,0.0331],[-32,0.0373],[-30,0.0368],[-28,0.0383],[-26,0.0404],[-24,0.0388],[-22,0.0443],[-20,0.0443],[-18,0.0428],[-16,0.0457],[-14,0.0407],[-12,0.0453],[-10,0.0405],[-8,0.0498],[-6,0.0454],[-4,0.0441],[-2,0.0488],[0,0.0433],[2,0.0453],[4,0.045],[6,0.0403],[8,0.0442],[10,0.0468],[12,0.0518],[14,0.0482],[16,0.0492],[18,0.051],[20,0.049],[22,0.0472],[24,0.0465],[26,0.0473],[28,0.0455],[30,0.0477],[32,0.0462],[34,0.0461],[36,0.0486],[38,0.0473],[42,0.0471],[44,0.0434],[46,0.0431],[48,0.0423],[50,0.0421],[52,0.0432],[54,0.0459],[56,0.0455],[58,0.0454],[60,0.0444],[62,0.0475],[64,0.0442],[66,0.0398],[68,0.0441],[70,0.0436],[72,0.0399],[74,0.042],[76,0.047],[78,0.0495],[80,0.0456],[82,0.0482],[84,0.0487],[86,0.0505],[90,0.051],[92,0.0478],[94,0.0463],[96,0.0422],[98,0.0375],[100,0.0353],[102,0.0371],[104,0.0436],[106,0.0346],[108,0.0379],[110,0.0325],[112,0.0295],[114,0.0158],[116,0.0027],[118,0.0114],[120,0.0143],[122,0.0151],[124,0.0184],[126,0.0184],[128,0.0192],[130,0.0192],[132,0.0267],[134,0.0126],[136,0.0149],[138,0.0235],[140,0.016],[142,0.0207],[144,0.0166],[146,0.0068],[148,0.0119],[150,0.0277],[152,0.0196],[154,0.024],[156,0.0232],[158,0.0094],[160,0.01],[162,0.0191],[164,0.0105],[168,0.0098],[170,0.0104],[172,0.0158],[174,0.0144],[176,0.0002],[178,0.0081],[180,0.0052],[182,0.0051],[184,0.0052],[186,0.0036],[188,0],[194,0],[196,0.0049],[198,0.0152],[200,0.0171],[202,0.0081],[204,0.0047],[206,0],[210,0],[212,0.0026],[214,0.0098],[216,0.0006],[218,0.0094],[220,0.0048],[222,0.0191],[224,0.0096],[226,0.0083],[228,0.0003],[230,0.0135],[232,0.009],[234,0.0005],[236,0.0123],[238,0.0056],[240,0.0121],[242,0.0134],[244,0.0132],[246,0.0046],[248,0.0044],[250,0.0002],[252,0],[254,0.0045],[256,0.0001],[258,0.0049],[260,0.005],[262,0.0008],[264,0.0002],[266,0.0054],[268,0.0245],[270,0.0151],[272,0.01],[274,0.033],[276,0.0418],[278,0.0003],[280,0.0425],[282,0.028],[284,0.0311],[286,0.0324],[288,0.0243],[290,0.0177],[292,0.0134],[294,0.0257],[296,0.0092],[298,0.0138],[300,0.0182],[302,0.0234],[304,0.02],[306,0.0237],[308,0.0189],[310,0.0222],[312,0.0098],[314,0.0204],[316,0.0207],[318,0.0339],[320,0.0153],[322,0.0212],[324,0.0352],[326,0.0238],[328,0.0213],[330,0.0255],[332,0.0252],[334,0.0294],[336,0.0224],[338,0.0305],[342,0.0327],[344,0.0242],[346,0.0124],[348,0.0065],[350,0.0115],[352,0.0141],[354,0.0163],[356,0.0202],[358,0.0178],[360,0.0295],[362,0.0624],[364,0.0641],[366,0.0628],[368,0.0364],[370,0.0612],[372,0.098],[374,0.1035],[376,0.1039],[378,0.1112],[380,0.1125],[382,0.0929],[384,0.0616],[386,0.0318],[388,0.0331],[390,0.0289],[392,0.0691],[394,0.1183],[396,0.1064],[398,0.1059],[400,0.1111],[402,0.1904],[404,0.1261],[406,0.1228],[408,0.1131],[410,0.0949],[412,0.0941],[414,0.1221],[416,0.0964],[418,0.0736],[420,0.0564],[422,0.0322],[424,0.0392],[426,0.0307],[428,0.0201],[430,0.0259],[432,0.0332],[434,0.0255],[436,0.0347],[438,0.0347],[440,0.034],[442,0.023],[444,0.0242],[446,0.043],[448,0.0323],[450,0.0275],[452,0.0329],[454,0.0409],[456,0.0218],[458,0.0497],[460,0.0722],[462,0.0978],[464,0.1202],[466,0.184],[468,0.2014],[470,0.1721],[472,0.0436],[474,0.0232],[476,0.047],[478,0.0602],[480,0.044],[482,0.0852],[484,0.0972],[486,0.1273],[488,0.1775],[490,0.2339],[492,0.2556],[494,0.2358],[496,0.1587],[498,0.1224],[500,0.141],[502,0.1619],[504,0.2406],[506,0.2917],[508,0.2786],[510,0.258],[512,0.2715],[514,0.2408],[516,0.1309],[518,0.1381],[520,0.2005],[522,0.2027],[524,0.1756],[526,0.2101],[528,0.228],[530,0.2219],[532,0.2238],[534,0.1933],[536,0.1484],[538,0.1999],[540,0.2123],[542,0.2202],[544,0.2221],[546,0.2287],[548,0.247],[550,0.2579],[552,0.2469],[554,0.2435],[556,0.2409],[558,0.2262],[560,0.2328],[562,0.2355],[564,0.2181],[566,0.2329],[568,0.2352],[570,0.2259],[572,0.2524],[574,0.2474],[576,0.245],[578,0.2401],[580,0.2373],[582,0.2335],[584,0.208],[586,0.181],[588,0.1797],[590,0.2168],[592,0.2337],[594,0.2344],[596,0.2097],[598,0.1922],[600,0.1312],[602,0.1209],[604,0.0385],[606,0.0349],[608,0.0255],[610,0.0248],[614,0.0287],[616,0.0272],[618,0.028],[620,0.0284],[622,0.0245],[624,0.0245],[626,0.0218],[628,0.0277],[630,0.0273],[632,0.0262],[634,0.0234],[636,0.02],[638,0.0273],[640,0.0262],[642,0.026],[644,0.0156],[646,0.0192],[648,0.0291],[650,0.0302],[652,0.0255],[654,0.0219],[656,0.0253],[658,0.0249],[660,0.0239],[662,0.0271],[664,0.029],[666,0.0247],[668,0.0244],[670,0.0207],[672,0.0253],[674,0.0225],[676,0.0284],[678,0.0216],[680,0.0241],[682,0.0178],[684,0.0209],[686,0.0214],[688,0.0217],[694,0.02],[696,0.0175],[698,0.0195],[700,0.0141],[702,0.0242],[704,0.0181],[706,0.0218],[708,0.0201],[710,0.0226],[712,0.0246],[714,0.0212],[716,0.0221],[718,0.0212],[720,0.0212]]
       }
       CTRL_C_tongue_roll.X {type=MovieSceneScriptingFloatChannel, default=0, const [0,492]}
       CTRL_C_tongue_roll.Y {type=MovieSceneScriptingFloatChannel, default=0, const [0,492]}
       CTRL_C_tongue.X {type=MovieSceneScriptingFloatChannel, default=0, const [0,492]}
       CTRL_C_tongue.Y {type=MovieSceneScriptingFloatChannel, default=0, const [0,492]}
       CTRL_C_tongue_narrowWide {type=MovieSceneScriptingFloatChannel, default=0, const [0,492]}
       CTRL_C_tongue_inOut {type=MovieSceneScriptingFloatChannel, default=0, const [0,492]}
       CTRL_R_nose.X {
        type=MovieSceneScriptingFloatChannel
        default=0
        thinned_from=492
        keys [[-262,0.0007],[-256,0.0065],[-254,0.0074],[-252,0.0077],[-250,0.007],[-248,0.0075],[-244,0.0072],[-242,0.0065],[-240,0.0065],[-238,0.0062],[-236,0.0056],[-234,0.0056],[-230,0.006],[-228,0.0058],[-224,0.006],[-222,0.0069],[-220,0.0072],[-218,0.0077],[-216,0.0098],[-214,0.0114],[-212,0.0116],[-208,0.0126],[-204,0.0147],[-202,0.0162],[-200,0.0164],[-198,0.0158],[-196,0.0159],[-192,0.017],[-190,0.0171],[-188,0.0141],[-184,0.0137],[-182,0.0127],[-180,0.0105],[-178,0.0059],[-176,0.0007],[-172,-0.0113],[-170,-0.0177],[-168,-0.0245],[-166,-0.032],[-164,-0.0401],[-162,-0.0448],[-160,-0.0538],[-158,-0.0557],[-156,-0.0718],[-154,-0.0862],[-152,-0.0957],[-150,-0.1093],[-148,-0.1136],[-146,-0.1109],[-142,-0.1168],[-140,-0.1255],[-138,-0.1268],[-136,-0.1262],[-134,-0.1274],[-132,-0.1256],[-130,-0.1248],[-126,-0.1199],[-124,-0.1152],[-122,-0.1033],[-120,-0.0859],[-118,-0.0577],[-116,-0.0367],[-114,-0.0312],[-112,-0.0194],[-110,-0.0145],[-108,-0.0118],[-106,-0.0099],[-104,-0.0091],[-102,-0.0055],[-100,-0.0051],[-98,-0.0051],[-96,-0.004],[-94,-0.0055],[-92,-0.006],[-90,-0.0093],[-88,-0.0102],[-86,-0.0131],[-84,-0.0181],[-82,-0.0261],[-80,-0.0378],[-78,-0.055],[-76,-0.0603],[-74,-0.091],[-72,-0.1014],[-70,-0.1159],[-68,-0.114],[-66,-0.1008],[-64,-0.0688],[-62,-0.0611],[-60,-0.0279],[-58,-0.0245],[-56,-0.017],[-54,-0.013],[-52,-0.011],[-50,-0.0087],[-48,-0.0078],[-46,0.0014],[-44,0.0058],[-42,0.0127],[-40,0.0159],[-38,0.0179],[-36,0.0184],[-34,0.0214],[-32,0.0232],[-30,0.0233],[-28,0.0239],[-26,0.0237],[-24,0.023],[-22,0.0227],[-20,0.0218],[-18,0.0223],[-16,0.0212],[-14,0.0209],[-12,0.0211],[-10,0.0196],[-8,0.02],[-6,0.0198],[-4,0.0201],[-2,0.0198],[0,0.0187],[4,0.0178],[6,0.0162],[8,0.0144],[10,0.013],[12,0.0095],[14,0.0086],[16,0.0082],[20,0.0055],[22,0.0048],[24,0.0044],[26,0.0037],[28,0.0032],[30,0.002],[32,0.0004],[36,-0.001],[38,-0.002],[44,-0.0031],[50,-0.0027],[54,-0.0034],[58,-0.0051],[62,-0.0053],[64,-0.0047],[66,-0.0055],[72,-0.0056],[74,-0.0063],[80,-0.0062],[82,-0.0048],[84,-0.0043],[86,-0.004],[88,-0.0027],[90,-0.0022],[92,0.0046],[94,0.0043],[96,0.0079],[100,0.0104],[102,0.0131],[106,0.0159],[108,0.0148],[110,0.0144],[112,0.0142],[116,0.0116],[118,0.0106],[120,0.0104],[122,0.01],[124,0.0089],[130,0.0088],[132,0.0083],[134,0.0085],[136,0.0081],[138,0.0064],[140,0.0051],[142,0.0046],[144,0.0037],[146,0.0035],[148,0.002],[150,0.001],[152,-0.0004],[154,-0.0007],[158,-0.0009],[160,-0.0018],[164,-0.0029],[168,-0.0013],[170,-0.0016],[174,-0.0015],[176,-0.001],[178,-0.0008],[180,-0.0016],[182,-0.0013],[190,-0.0012],[196,-0.0028],[198,-0.003],[200,-0.0023],[202,-0.0019],[204,-0.0019],[206,-0.0016],[208,0],[210,0],[212,0.0008],[216,0.0016],[220,0.0019],[222,0.0024],[224,0.0037],[226,0.0054],[228,0.0066],[230,0.0068],[232,0.0073],[234,0.0092],[236,0.013],[238,0.0148],[240,0.0158],[242,0.0157],[246,0.0135],[248,0.0116],[250,0.0116],[252,0.0103],[254,0.01],[256,0.0094],[258,0.0096],[262,0.0067],[266,0.0067],[268,0.0103],[270,0.0121],[272,0.0125],[274,0.0136],[276,0.0092],[278,0.0016],[280,-0.0036],[282,-0.0053],[284,-0.0067],[286,-0.0068],[290,-0.0055],[292,-0.0047],[294,-0.0053],[296,-0.0073],[298,-0.0061],[300,-0.0055],[304,-0.0067],[306,-0.0071],[308,-0.0079],[310,-0.0071],[312,-0.0065],[314,-0.0069],[316,-0.0079],[318,-0.0086],[320,-0.0085],[322,-0.0086],[324,-0.008],[328,-0.0062],[332,-0.0053],[334,-0.0058],[336,-0.0055],[338,-0.0041],[340,-0.0042],[342,-0.0032],[344,-0.0026],[348,-0.0003],[352,0.004],[354,0.0043],[358,0.0063],[360,0.0069],[362,0.0071],[364,0.008],[368,0.0088],[370,0.0099],[376,0.0104],[380,0.0127],[382,0.0127],[384,0.0187],[386,0.0227],[388,0.0253],[390,0.0247],[392,0.0267],[394,0.0332],[396,0.0405],[398,0.0452],[400,0.0463],[402,0.0476],[404,0.0435],[406,0.0425],[408,0.0427],[410,0.0435],[412,0.0438],[414,0.043],[416,0.0415],[418,0.0406],[420,0.0378],[422,0.0363],[424,0.0366],[426,0.0359],[428,0.0358],[430,0.0347],[432,0.0334],[434,0.0333],[436,0.0327],[438,0.0314],[440,0.0313],[442,0.0308],[444,0.0295],[446,0.0289],[450,0.0292],[452,0.0298],[454,0.0295],[456,0.03],[460,0.0321],[462,0.0322],[464,0.0331],[466,0.0342],[468,0.0342],[472,0.0298],[474,0.0277],[476,0.0266],[478,0.0206],[480,0.0157],[482,0.0127],[484,0.0123],[486,0.0102],[488,0.0046],[490,0.0028],[492,-0.0026],[494,-0.0112],[496,-0.0151],[498,-0.0125],[500,-0.0079],[502,-0.0092],[504,-0.0079],[506,-0.01],[508,-0.0106],[510,-0.0104],[512,-0.0098],[514,-0.0044],[516,0.0007],[518,0.0011],[520,0.0043],[522,0.0052],[524,0.0058],[526,0.0031],[530,0.0014],[532,0.0026],[534,0.0018],[536,0.0035],[538,0.0042],[540,0.0042],[542,0.0008],[544,0.0037],[546,0.0044],[548,0.0085],[550,0.009],[552,0.0089],[554,0.0085],[558,0.0073],[560,0.0065],[564,0.0067],[566,0.0063],[570,0.0061],[572,0.0066],[576,0.0062],[578,0.0067],[580,0.0067],[582,0.007],[584,0.0095],[586,0.0113],[588,0.0126],[590,0.0112],[592,0.0087],[594,0.0084],[596,0.0097],[598,0.013],[600,0.0199],[602,0.0202],[604,0.0214],[608,0.021],[610,0.0211],[612,0.0206],[616,0.0208],[620,0.0205],[622,0.0184],[626,0.0178],[628,0.0172],[630,0.0175],[632,0.0179],[634,0.0178],[636,0.0182],[638,0.0183],[640,0.017],[642,0.0177],[646,0.0169],[648,0.0167],[650,0.0171],[652,0.0158],[656,0.0144],[660,0.0145],[664,0.0139],[668,0.0136],[672,0.0129],[674,0.0128],[676,0.0122],[680,0.0118],[684,0.0125],[686,0.0126],[688,0.0132],[690,0.0125],[692,0.0121],[694,0.0129],[696,0.0127],[698,0.0117],[700,0.0117],[702,0.0112],[704,0.011],[706,0.0095],[712,0.0092],[714,0.0095],[716,0.0096],[718,0.0102],[720,0.0102]]
       }
       CTRL_R_nose.Y {
        type=MovieSceneScriptingFloatChannel
        default=0
        thinned_from=492
        keys [[-262,0.0098],[-260,0.0059],[-258,0.0041],[-256,0.0048],[-254,0.0007],[-252,-0.0001],[-250,-0.0013],[-248,-0.0019],[-246,-0.0013],[-244,-0.0012],[-242,-0.0044],[-240,-0.004],[-238,-0.0048],[-236,-0.0047],[-234,-0.0036],[-232,0.0042],[-230,0.0065],[-226,-0.0014],[-224,-0.0006],[-222,0.0009],[-220,0.0011],[-216,0.0107],[-214,0.0115],[-212,0.0117],[-210,0.0155],[-208,0.0162],[-206,0.0192],[-204,0.0166],[-202,0.0095],[-200,0.0128],[-198,0.0097],[-196,0.01],[-194,0.0097],[-192,0.0068],[-190,0.0059],[-188,0.0072],[-186,0.0073],[-184,0.0088],[-182,0.0105],[-180,0.0143],[-178,0.0241],[-176,0.0336],[-174,0.0428],[-172,0.0565],[-168,0.086],[-164,0.1028],[-162,0.1062],[-160,0.1143],[-158,0.1142],[-156,0.1133],[-154,0.1026],[-152,0.0922],[-150,0.0822],[-148,0.0672],[-146,0.058],[-144,0.0474],[-142,0.0466],[-138,0.0357],[-136,0.0368],[-132,0.0384],[-130,0.0389],[-128,0.0398],[-126,0.0385],[-122,0.0348],[-120,0.0363],[-118,0.0355],[-116,0.0365],[-112,0.0361],[-110,0.0375],[-108,0.0375],[-106,0.0371],[-104,0.0371],[-102,0.0273],[-100,0.0251],[-98,0.025],[-96,0.0223],[-94,0.0224],[-92,0.0215],[-90,0.0215],[-88,0.0221],[-86,0.024],[-84,0.0303],[-82,0.0427],[-80,0.0631],[-78,0.0785],[-76,0.0811],[-74,0.0833],[-72,0.0748],[-70,0.0712],[-68,0.0732],[-66,0.0734],[-64,0.0637],[-62,0.0595],[-60,0.039],[-58,0.0367],[-56,0.0335],[-54,0.0361],[-52,0.0379],[-50,0.0359],[-48,0.0346],[-46,0.011],[-44,0.0046],[-42,-0.0142],[-40,-0.0066],[-38,-0.0063],[-36,-0.0036],[-34,-0.0013],[-32,0.0061],[-30,0.0112],[-28,0.0155],[-26,0.0194],[-24,0.0253],[-22,0.0264],[-20,0.0299],[-18,0.0278],[-16,0.0303],[-10,0.0366],[-8,0.0355],[-6,0.0381],[-4,0.0371],[-2,0.0377],[0,0.0404],[2,0.0406],[6,0.0391],[8,0.0396],[10,0.0383],[12,0.0386],[14,0.0383],[16,0.0378],[18,0.0358],[20,0.0345],[22,0.0364],[24,0.0365],[26,0.037],[28,0.0371],[30,0.0369],[32,0.0364],[34,0.0361],[36,0.0388],[38,0.0391],[40,0.04],[42,0.0411],[44,0.0425],[46,0.0426],[48,0.0414],[50,0.0378],[52,0.0387],[54,0.0394],[56,0.0405],[58,0.0423],[60,0.043],[62,0.042],[64,0.039],[66,0.0403],[68,0.0428],[70,0.0422],[72,0.0386],[74,0.04],[76,0.0422],[78,0.0426],[80,0.0404],[82,0.0376],[84,0.0372],[86,0.0382],[90,0.0368],[92,0.0285],[94,0.0298],[96,0.0231],[98,0.0195],[100,0.0132],[102,0.0043],[104,0.0028],[106,0.0035],[108,0.0039],[110,0.0026],[112,-0.0028],[114,-0.0059],[116,-0.0079],[118,-0.008],[120,-0.0078],[124,-0.0054],[126,-0.0052],[128,-0.0025],[130,-0.0025],[132,0.0011],[134,-0.0002],[136,-0.0016],[138,-0.001],[144,0.0035],[146,0.0035],[148,0.0059],[150,0.0059],[152,0.0075],[154,0.0097],[156,0.0096],[158,0.0076],[160,0.0106],[166,0.0098],[168,0.0061],[170,0.0071],[172,0.0099],[174,0.0099],[176,0.0113],[178,0.0082],[180,0.0106],[182,0.0078],[184,0.0108],[186,0.0109],[188,0.0113],[190,0.011],[192,0.0115],[194,0.0141],[198,0.0164],[200,0.0186],[202,0.0179],[204,0.0176],[206,0.0191],[208,0.0159],[210,0.0157],[212,0.0146],[216,0.0145],[218,0.0127],[220,0.0091],[222,0.0088],[224,0.0052],[226,-0.0032],[228,-0.01],[230,-0.0137],[232,-0.0192],[234,-0.0309],[236,-0.0553],[238,-0.0694],[240,-0.0798],[242,-0.0823],[244,-0.0819],[246,-0.0874],[248,-0.0825],[250,-0.0823],[252,-0.0812],[254,-0.0811],[256,-0.0784],[260,-0.0697],[262,-0.0706],[264,-0.069],[266,-0.0669],[268,-0.0781],[270,-0.093],[272,-0.1056],[274,-0.0934],[276,-0.0563],[278,-0.0349],[280,-0.0256],[282,-0.0265],[284,-0.0211],[286,-0.0192],[288,-0.0191],[290,-0.0234],[292,-0.0295],[294,-0.0304],[296,-0.0251],[298,-0.025],[300,-0.0263],[302,-0.026],[304,-0.0286],[306,-0.0269],[308,-0.0195],[310,-0.0195],[314,-0.0177],[316,-0.0135],[320,-0.0106],[322,-0.0042],[324,0.0011],[326,0.0032],[328,0.0011],[330,0.0039],[332,0.0107],[334,0.0146],[336,0.0141],[338,0.0148],[340,0.0186],[342,0.0201],[344,0.017],[348,0.0026],[350,-0.0063],[352,-0.0098],[354,-0.0093],[356,-0.008],[358,-0.0096],[360,-0.0096],[362,-0.0117],[364,-0.0162],[366,-0.0192],[368,-0.0227],[370,-0.0244],[372,-0.0243],[374,-0.023],[376,-0.0229],[378,-0.0286],[380,-0.0309],[382,-0.0355],[384,-0.0819],[386,-0.1142],[388,-0.1502],[390,-0.1474],[392,-0.1612],[394,-0.1977],[396,-0.2289],[398,-0.2388],[400,-0.2397],[402,-0.233],[404,-0.2195],[406,-0.1911],[408,-0.1817],[410,-0.1714],[412,-0.1659],[414,-0.153],[416,-0.143],[418,-0.1351],[420,-0.1163],[422,-0.1075],[424,-0.1085],[426,-0.1079],[428,-0.1093],[430,-0.1031],[432,-0.0984],[434,-0.0944],[436,-0.0923],[438,-0.0889],[440,-0.0837],[442,-0.0822],[444,-0.0784],[446,-0.0756],[448,-0.0754],[450,-0.0745],[452,-0.073],[454,-0.0726],[456,-0.0749],[458,-0.0778],[460,-0.0758],[462,-0.0727],[464,-0.0761],[466,-0.09],[468,-0.0966],[470,-0.0943],[472,-0.0711],[474,-0.0665],[476,-0.0565],[478,-0.0522],[480,-0.0401],[482,-0.0292],[484,-0.0277],[486,-0.0271],[488,-0.0172],[490,-0.0109],[492,-0.0091],[494,-0.0039],[496,-0.0047],[498,-0.0029],[500,-0.007],[502,0.0004],[504,0.004],[506,0.0132],[508,0.0184],[510,0.0226],[512,0.0224],[514,0.0113],[516,0.0105],[518,0.0126],[520,0.0416],[522,0.0737],[524,0.088],[526,0.109],[528,0.1109],[530,0.1077],[532,0.1037],[534,0.0979],[536,0.084],[538,0.0528],[540,0.0471],[542,0.0379],[544,0.0149],[546,0.0027],[548,-0.0074],[550,-0.0091],[552,-0.0101],[554,-0.0045],[556,-0.0007],[558,0],[560,0.0005],[562,0.0003],[566,-0.0016],[572,-0.0016],[574,0.001],[576,0.0046],[578,0.0066],[580,0.0057],[582,0.0036],[584,0.0085],[586,0.0147],[588,0.02],[590,0.0206],[592,0.0166],[594,0.0156],[596,0.0048],[598,-0.0002],[600,-0.0134],[602,-0.0136],[604,-0.0134],[606,-0.014],[608,-0.0158],[610,-0.0161],[612,-0.0146],[614,-0.0132],[618,-0.01],[620,-0.0095],[622,-0.0049],[626,-0.0036],[628,-0.001],[630,-0.0049],[632,-0.0046],[636,-0.0029],[638,-0.0027],[642,-0.0013],[644,-0.0016],[646,-0.003],[648,0.0002],[650,0.002],[652,0.003],[654,0.0046],[656,0.0074],[658,0.0089],[660,0.0082],[662,0.0107],[664,0.0123],[666,0.012],[668,0.0126],[670,0.0138],[672,0.0132],[674,0.0134],[676,0.0141],[678,0.013],[680,0.0123],[682,0.0125],[690,0.0104],[692,0.0094],[694,0.0067],[696,0.0057],[698,0.0062],[700,0.0075],[702,0.0093],[706,0.0105],[708,0.0108],[710,0.0123],[712,0.0126],[714,0.0116],[716,0.0136],[718,0.0126],[720,0.0126]]
       }
       CTRL_L_nose.X {
        type=MovieSceneScriptingFloatChannel
        default=0
        thinned_from=492
        keys [[-262,-0.012],[-260,-0.0106],[-258,-0.0081],[-256,-0.006],[-254,-0.005],[-252,-0.0047],[-250,-0.0056],[-248,-0.005],[-246,-0.0054],[-244,-0.0056],[-242,-0.0067],[-240,-0.0062],[-238,-0.0067],[-236,-0.0077],[-234,-0.0077],[-230,-0.0072],[-228,-0.0074],[-226,-0.0071],[-224,-0.007],[-222,-0.0063],[-218,-0.0056],[-214,-0.0016],[-210,-0.0009],[-206,0.0013],[-204,0.0026],[-202,0.0043],[-200,0.0041],[-198,0.0032],[-196,0.0033],[-192,0.0043],[-190,0.0044],[-188,0.0019],[-186,0.0018],[-184,0.0023],[-182,0.001],[-180,-0.0018],[-178,-0.0071],[-176,-0.0128],[-174,-0.019],[-172,-0.0249],[-170,-0.0315],[-168,-0.0389],[-166,-0.0467],[-164,-0.0536],[-162,-0.0581],[-160,-0.0674],[-158,-0.0693],[-156,-0.0865],[-154,-0.1024],[-152,-0.1138],[-150,-0.1305],[-148,-0.136],[-146,-0.1347],[-144,-0.1377],[-142,-0.1416],[-140,-0.1504],[-138,-0.1523],[-136,-0.1514],[-134,-0.1533],[-132,-0.151],[-130,-0.1502],[-126,-0.1452],[-124,-0.1402],[-122,-0.1277],[-120,-0.1085],[-118,-0.0761],[-116,-0.0515],[-114,-0.0453],[-112,-0.032],[-110,-0.0265],[-108,-0.0234],[-106,-0.0214],[-104,-0.0206],[-102,-0.0164],[-100,-0.0161],[-98,-0.0161],[-96,-0.0143],[-94,-0.0157],[-92,-0.0164],[-90,-0.0203],[-88,-0.0213],[-86,-0.0247],[-84,-0.0301],[-82,-0.0383],[-80,-0.0513],[-78,-0.0699],[-76,-0.0756],[-74,-0.1109],[-72,-0.1231],[-70,-0.1387],[-68,-0.1368],[-66,-0.1213],[-64,-0.0858],[-62,-0.0771],[-60,-0.0397],[-58,-0.036],[-56,-0.0281],[-54,-0.0239],[-52,-0.0218],[-50,-0.0192],[-48,-0.0183],[-46,-0.0095],[-44,-0.0046],[-42,0.0021],[-40,0.0047],[-38,0.0069],[-36,0.0076],[-34,0.0106],[-32,0.0124],[-30,0.0125],[-28,0.0131],[-22,0.0127],[-20,0.0119],[-18,0.0125],[-16,0.0114],[-12,0.0116],[-10,0.01],[-8,0.0108],[-6,0.0102],[-4,0.0105],[-2,0.0101],[0,0.009],[4,0.0081],[6,0.0069],[8,0.0048],[10,0.0033],[12,-0.0002],[16,-0.0013],[18,-0.0027],[20,-0.0038],[26,-0.0062],[28,-0.0069],[32,-0.0092],[34,-0.01],[36,-0.0102],[38,-0.0112],[40,-0.0117],[44,-0.0123],[50,-0.0125],[54,-0.0132],[58,-0.0148],[62,-0.0145],[64,-0.014],[66,-0.0147],[68,-0.015],[72,-0.0147],[74,-0.0155],[80,-0.0154],[82,-0.014],[84,-0.0133],[86,-0.0132],[88,-0.0119],[90,-0.0114],[92,-0.0049],[94,-0.0051],[96,-0.0012],[98,-0.0002],[100,0.0012],[102,0.0042],[106,0.0071],[108,0.0058],[110,0.0047],[112,0.0046],[116,0.0014],[118,0],[122,-0.0004],[124,-0.0014],[126,-0.0015],[128,-0.0018],[130,-0.0018],[132,-0.0027],[134,-0.0016],[136,-0.0023],[138,-0.0037],[140,-0.0053],[146,-0.007],[148,-0.0089],[150,-0.01],[152,-0.0119],[154,-0.0116],[158,-0.0117],[162,-0.0131],[164,-0.0136],[168,-0.0121],[170,-0.0121],[172,-0.0109],[174,-0.0111],[176,-0.0109],[178,-0.0111],[180,-0.0119],[190,-0.0116],[194,-0.0123],[196,-0.0125],[198,-0.0124],[202,-0.0115],[204,-0.0114],[206,-0.0116],[208,-0.0101],[210,-0.0101],[212,-0.0087],[214,-0.0085],[216,-0.0079],[218,-0.0078],[220,-0.0081],[222,-0.0077],[224,-0.0065],[226,-0.0048],[228,-0.0034],[232,-0.0028],[234,-0.0008],[236,0.0021],[238,0.0035],[240,0.0042],[242,0.0045],[246,0.002],[248,0.0005],[250,0.0007],[252,-0.001],[254,-0.001],[256,-0.0017],[258,-0.0014],[260,-0.0017],[262,-0.0026],[264,-0.0033],[266,-0.0029],[268,-0.0004],[270,0.0013],[272,0.0007],[274,0.0019],[276,-0.0015],[278,-0.0084],[280,-0.0129],[282,-0.0146],[284,-0.0155],[286,-0.0152],[292,-0.0132],[294,-0.0144],[296,-0.0162],[300,-0.0144],[302,-0.0148],[304,-0.0154],[306,-0.0156],[308,-0.0163],[312,-0.015],[314,-0.0153],[316,-0.0162],[318,-0.0169],[320,-0.0168],[322,-0.0171],[326,-0.0157],[328,-0.0146],[330,-0.0139],[332,-0.0127],[334,-0.0134],[336,-0.0135],[338,-0.0122],[340,-0.0123],[342,-0.0112],[344,-0.0108],[346,-0.0099],[348,-0.0093],[350,-0.0078],[352,-0.006],[354,-0.0056],[358,-0.0038],[360,-0.0032],[362,-0.0031],[364,-0.0025],[368,-0.0022],[370,-0.0013],[372,-0.0012],[376,-0.0013],[378,-0.0001],[380,0],[382,-0.0002],[384,0.0059],[386,0.011],[388,0.0168],[390,0.0165],[392,0.018],[396,0.0301],[398,0.0336],[402,0.0353],[404,0.0323],[406,0.03],[408,0.0296],[412,0.0304],[414,0.0299],[418,0.0273],[420,0.0248],[422,0.0235],[424,0.0238],[426,0.0232],[428,0.0236],[430,0.0227],[432,0.0216],[434,0.021],[436,0.0207],[438,0.0196],[440,0.019],[444,0.0174],[446,0.017],[450,0.0174],[452,0.018],[454,0.0178],[456,0.0188],[458,0.0201],[462,0.0204],[464,0.0209],[466,0.0217],[468,0.0216],[470,0.0205],[472,0.0202],[476,0.0185],[478,0.0129],[480,0.0071],[482,0.0027],[484,0.002],[486,-0.0016],[488,-0.0082],[490,-0.0101],[492,-0.0158],[494,-0.0252],[496,-0.0303],[498,-0.0276],[500,-0.0232],[502,-0.0242],[504,-0.0227],[506,-0.0245],[508,-0.0247],[510,-0.0243],[512,-0.0234],[516,-0.0137],[518,-0.0133],[520,-0.0095],[524,-0.007],[526,-0.0098],[528,-0.0104],[530,-0.0113],[532,-0.0102],[534,-0.0104],[536,-0.0076],[538,-0.007],[540,-0.0071],[542,-0.0112],[544,-0.0088],[546,-0.0084],[548,-0.0048],[550,-0.0045],[554,-0.0044],[556,-0.0046],[558,-0.0051],[560,-0.006],[562,-0.0059],[564,-0.0061],[566,-0.0066],[570,-0.0067],[572,-0.0057],[576,-0.0057],[578,-0.0053],[580,-0.0052],[582,-0.0046],[584,-0.002],[586,0.0003],[588,0.0017],[590,0.0002],[592,-0.0022],[594,-0.0025],[596,-0.0014],[598,0.0024],[600,0.0096],[602,0.0098],[604,0.0103],[614,0.0095],[616,0.0096],[620,0.0092],[622,0.0075],[624,0.0071],[626,0.007],[628,0.0064],[630,0.0062],[632,0.0069],[634,0.0069],[636,0.0073],[638,0.0071],[640,0.006],[642,0.0068],[646,0.0061],[648,0.006],[650,0.0063],[652,0.005],[654,0.0049],[656,0.0044],[664,0.0039],[666,0.0034],[668,0.0036],[670,0.0033],[672,0.0026],[674,0.0027],[676,0.0017],[678,0.0023],[680,0.0016],[684,0.0026],[688,0.0028],[690,0.0024],[692,0.0022],[694,0.0025],[696,0.0022],[698,0.0016],[700,0.0018],[702,0.0013],[704,0.0012],[706,-0.0002],[710,0.0002],[712,0.0002],[714,-0.0002],[716,-0.0003],[718,0.0003],[720,0.0003]]
       }
       CTRL_L_nose.Y {
        type=MovieSceneScriptingFloatChannel
        default=0
        thinned_from=492
        keys [[-262,0.0149],[-260,0.0164],[-258,0.0164],[-256,0.0143],[-254,0.0105],[-252,0.0089],[-250,0.0063],[-248,0.0054],[-246,0.0072],[-244,0.0077],[-242,0.0062],[-240,0.0057],[-238,0.0039],[-236,0.0038],[-234,0.0045],[-232,0.0095],[-230,0.0101],[-228,0.0068],[-226,0.0043],[-224,0.0046],[-220,0.0074],[-218,0.0106],[-216,0.0153],[-214,0.0173],[-212,0.0177],[-210,0.0215],[-208,0.0214],[-206,0.0234],[-204,0.0214],[-202,0.019],[-200,0.0225],[-198,0.0222],[-196,0.0248],[-194,0.0262],[-192,0.0258],[-190,0.0258],[-188,0.0283],[-186,0.0287],[-184,0.0316],[-182,0.036],[-180,0.0413],[-178,0.0499],[-174,0.0667],[-172,0.0744],[-170,0.0838],[-168,0.0938],[-166,0.0978],[-164,0.1005],[-162,0.0982],[-160,0.0902],[-158,0.0897],[-156,0.0844],[-154,0.0718],[-152,0.0666],[-150,0.062],[-148,0.0535],[-146,0.0493],[-144,0.0413],[-142,0.0421],[-140,0.0346],[-138,0.0301],[-136,0.0274],[-134,0.0322],[-132,0.0322],[-130,0.034],[-128,0.034],[-126,0.0288],[-124,0.0294],[-122,0.0293],[-120,0.028],[-118,0.0219],[-116,0.0165],[-114,0.0143],[-112,0.0095],[-110,0.0074],[-108,0.0058],[-106,0.0056],[-104,0.0057],[-102,-0.003],[-100,-0.0024],[-98,-0.0022],[-96,-0.0032],[-94,-0.0002],[-92,0.0004],[-90,0.006],[-88,0.008],[-86,0.0143],[-84,0.0211],[-82,0.0309],[-80,0.0434],[-78,0.054],[-76,0.0562],[-74,0.0632],[-72,0.0614],[-70,0.0586],[-68,0.0616],[-66,0.0626],[-64,0.0483],[-62,0.042],[-60,0.0167],[-58,0.0124],[-56,0.0059],[-52,0.0131],[-48,0.011],[-46,0.0015],[-44,-0.0001],[-42,-0.0099],[-40,-0.0122],[-38,-0.0136],[-36,-0.0138],[-34,-0.0166],[-32,-0.0147],[-30,-0.0135],[-26,-0.0079],[-24,-0.0045],[-22,-0.0034],[-20,-0.0015],[-18,-0.0019],[-16,0.0009],[-14,0.0014],[-12,0.0008],[-10,0.0021],[-8,0.0029],[-6,0.0024],[-4,0.0022],[-2,0.0027],[0,0.0057],[2,0.0062],[4,0.0057],[6,0.0058],[8,0.0063],[10,0.0056],[12,0.007],[14,0.0064],[16,0.0065],[20,0.0058],[22,0.0085],[24,0.01],[26,0.0105],[28,0.0093],[30,0.0077],[32,0.0091],[34,0.0092],[36,0.0098],[38,0.0084],[40,0.009],[42,0.0104],[46,0.0106],[48,0.0097],[50,0.0073],[52,0.0074],[54,0.0073],[56,0.0095],[58,0.0128],[60,0.0145],[62,0.0138],[64,0.0112],[66,0.0118],[68,0.0138],[70,0.0135],[72,0.0117],[74,0.0117],[76,0.0112],[78,0.0102],[80,0.0097],[84,0.0064],[86,0.0079],[88,0.0075],[90,0.0069],[92,0.0017],[94,0.0036],[96,0.0006],[100,-0.0035],[102,-0.0073],[104,-0.0095],[106,-0.0144],[108,-0.0128],[110,-0.013],[112,-0.0163],[114,-0.0176],[116,-0.0182],[118,-0.0171],[120,-0.0167],[122,-0.0167],[124,-0.0116],[126,-0.0114],[128,-0.0078],[130,-0.0078],[132,-0.0072],[134,-0.0053],[136,-0.0026],[138,-0.0007],[140,0.0017],[142,0.0035],[144,0.007],[146,0.0075],[148,0.0076],[150,0.0088],[152,0.0095],[154,0.0112],[156,0.011],[158,0.0083],[160,0.009],[162,0.0052],[164,0.0055],[166,0.0079],[168,0.006],[170,0.006],[172,0.0084],[174,0.0103],[176,0.0125],[178,0.0138],[180,0.0129],[182,0.0134],[184,0.0152],[186,0.0159],[188,0.0172],[190,0.0164],[192,0.0166],[196,0.0185],[198,0.0208],[200,0.0217],[202,0.0196],[204,0.0192],[206,0.0218],[208,0.0201],[210,0.0199],[212,0.0164],[214,0.0184],[216,0.0191],[218,0.0202],[220,0.0178],[222,0.0157],[224,0.0131],[226,0.0084],[228,0.0039],[230,0.0016],[232,-0.0025],[234,-0.0141],[236,-0.0318],[238,-0.0439],[240,-0.0492],[242,-0.0509],[244,-0.0513],[246,-0.0538],[250,-0.0547],[252,-0.0535],[254,-0.0528],[256,-0.0546],[258,-0.0531],[260,-0.0478],[262,-0.0478],[264,-0.0485],[266,-0.0469],[268,-0.0583],[270,-0.0801],[272,-0.0976],[274,-0.1114],[276,-0.095],[278,-0.0774],[280,-0.0623],[282,-0.0597],[284,-0.0539],[286,-0.0533],[288,-0.0541],[290,-0.0554],[292,-0.0564],[294,-0.0566],[296,-0.0547],[298,-0.057],[300,-0.0568],[302,-0.0542],[304,-0.054],[306,-0.0513],[308,-0.0494],[310,-0.0514],[314,-0.0492],[316,-0.0469],[318,-0.0414],[320,-0.0407],[322,-0.0385],[324,-0.0354],[326,-0.037],[328,-0.0353],[330,-0.0343],[332,-0.0276],[334,-0.0242],[336,-0.0248],[338,-0.0242],[340,-0.0195],[342,-0.0142],[344,-0.0153],[346,-0.0196],[348,-0.0227],[350,-0.0269],[352,-0.0295],[354,-0.029],[356,-0.027],[360,-0.0256],[362,-0.0224],[364,-0.0245],[366,-0.0257],[368,-0.0286],[370,-0.029],[372,-0.0288],[374,-0.0277],[376,-0.0276],[378,-0.0314],[380,-0.0329],[382,-0.0371],[384,-0.0603],[386,-0.0779],[388,-0.0827],[390,-0.0819],[392,-0.1072],[394,-0.1478],[396,-0.1793],[398,-0.1893],[400,-0.19],[402,-0.1822],[404,-0.1563],[406,-0.1403],[408,-0.138],[410,-0.1305],[412,-0.1244],[414,-0.1142],[416,-0.1015],[418,-0.095],[420,-0.0862],[422,-0.0772],[424,-0.0781],[426,-0.0771],[428,-0.0752],[430,-0.0726],[432,-0.0686],[434,-0.0643],[438,-0.0641],[440,-0.06],[442,-0.059],[444,-0.0587],[446,-0.0579],[448,-0.0584],[450,-0.0578],[454,-0.057],[456,-0.0593],[458,-0.0618],[460,-0.0598],[462,-0.0583],[464,-0.0575],[466,-0.0725],[468,-0.081],[470,-0.0823],[472,-0.0649],[474,-0.0547],[476,-0.0432],[478,-0.0393],[480,-0.0362],[482,-0.0353],[484,-0.0355],[486,-0.0377],[488,-0.0288],[490,-0.0161],[492,-0.0037],[494,0.0067],[496,0.0114],[498,0.018],[500,0.0162],[502,0.0196],[504,0.0215],[506,0.0254],[508,0.0281],[512,0.0306],[514,0.02],[516,0.0228],[518,0.0245],[520,0.0471],[522,0.0708],[524,0.0834],[526,0.0973],[528,0.0972],[530,0.094],[532,0.0925],[534,0.0867],[536,0.0766],[538,0.0523],[540,0.0478],[542,0.0388],[544,0.0246],[546,0.016],[548,0.0067],[550,0.0057],[552,0.0055],[554,0.0097],[556,0.0124],[558,0.0113],[562,0.0112],[564,0.0105],[566,0.011],[568,0.011],[570,0.0097],[572,0.0087],[574,0.0112],[576,0.0151],[578,0.0193],[580,0.0189],[582,0.0177],[584,0.0185],[586,0.0178],[588,0.02],[590,0.0256],[592,0.0277],[594,0.0277],[596,0.0214],[598,0.0163],[600,0.0062],[602,0.0056],[604,0.0015],[606,0.0004],[608,-0.0026],[610,-0.003],[612,-0.0023],[614,-0.0027],[616,-0.0034],[618,-0.0024],[620,-0.0019],[622,0.0015],[624,0.0019],[626,0.0029],[628,0.0028],[630,0.0002],[634,-0.0006],[636,0.0011],[638,0.0011],[640,0.0025],[642,0.0032],[646,0.0031],[648,0.0061],[650,0.007],[654,0.0069],[658,0.0088],[660,0.009],[662,0.0102],[664,0.011],[666,0.0109],[668,0.0111],[670,0.0121],[672,0.0114],[674,0.0111],[676,0.0131],[678,0.012],[680,0.0119],[682,0.0122],[684,0.0119],[686,0.0114],[688,0.0116],[690,0.0112],[692,0.012],[694,0.0114],[696,0.0101],[698,0.0107],[700,0.0118],[702,0.0139],[704,0.0143],[706,0.0156],[708,0.0158],[710,0.0168],[712,0.017],[714,0.0163],[716,0.0179],[718,0.0167],[720,0.0167]]
       }
       CTRL_L_eye.X {
        type=MovieSceneScriptingFloatChannel
        default=0
        thinned_from=492
        keys [[-262,0.0454],[-260,0.0394],[-258,0.0373],[-256,0.037],[-254,0.0252],[-252,0.0163],[-250,0.008],[-248,0.0028],[-246,-0.0017],[-244,-0.0029],[-242,0.0008],[-240,-0.0044],[-238,0.0005],[-236,0.0101],[-234,0.0109],[-232,0.0158],[-230,0.0198],[-228,0.0282],[-226,0.0333],[-224,0.0362],[-222,0.0397],[-220,0.0442],[-218,0.0448],[-216,0.0427],[-214,0.0415],[-212,0.0409],[-210,0.0316],[-208,0.029],[-206,0.0241],[-204,0.026],[-202,0.0242],[-200,0.0189],[-198,0.0165],[-194,0.0097],[-192,0.0076],[-190,0.0068],[-188,0.0027],[-186,0.0025],[-182,-0.0019],[-180,-0.0048],[-178,-0.0073],[-176,-0.0093],[-172,-0.0191],[-170,-0.0182],[-168,-0.0127],[-166,-0.0109],[-164,-0.0055],[-162,0.0006],[-160,0.0044],[-158,0.005],[-156,0.0101],[-154,0.0198],[-152,0.0193],[-150,0.0222],[-148,0.0209],[-146,0.018],[-144,0.0149],[-142,0.0134],[-140,0.002],[-138,-0.0002],[-136,-0.0004],[-134,-0.001],[-132,0.0028],[-130,0.0108],[-128,0.0103],[-126,0.0137],[-122,0.012],[-120,0.012],[-118,0.0148],[-116,0.0169],[-114,0.0163],[-112,0.0142],[-110,0.0124],[-108,0.0109],[-106,0.0081],[-104,0.0067],[-102,0.0043],[-100,-0.0048],[-98,-0.0059],[-96,-0.0058],[-94,-0.002],[-92,-0.0026],[-90,-0.0042],[-86,-0.0049],[-84,-0.0009],[-82,0.0004],[-80,-0.0016],[-76,0.0003],[-74,0.0088],[-72,0.0096],[-70,0.0113],[-68,0.0179],[-66,0.0217],[-64,0.0249],[-62,0.0258],[-58,0.0289],[-56,0.0311],[-54,0.0309],[-52,0.0287],[-50,0.0257],[-48,0.0251],[-46,0.0224],[-44,0.0689],[-42,0.0705],[-40,0.0505],[-38,0.0266],[-36,0.0214],[-34,0.0221],[-32,0.0201],[-30,0.0159],[-28,0.0103],[-26,0.0112],[-24,0.0107],[-22,0.0082],[-20,0.0085],[-18,0.0036],[-16,-0.1314],[-12,-0.1299],[-10,-0.1333],[-8,-0.1332],[-6,-0.1334],[-4,-0.1329],[-2,-0.131],[4,-0.1295],[6,-0.128],[10,-0.1268],[12,-0.1265],[14,-0.127],[16,-0.1271],[20,-0.1279],[22,-0.1281],[24,-0.128],[26,-0.1272],[28,-0.1276],[30,-0.1292],[32,-0.1301],[34,-0.129],[36,-0.1288],[40,-0.1291],[42,-0.1307],[48,-0.1299],[50,-0.1306],[54,-0.1288],[56,-0.129],[58,-0.1295],[60,-0.1313],[62,-0.1321],[64,-0.1308],[66,-0.1317],[68,-0.1335],[70,-0.1334],[72,-0.1322],[74,-0.1322],[76,-0.133],[78,-0.1331],[80,-0.1339],[82,-0.1343],[84,-0.1358],[86,-0.1377],[88,-0.1389],[90,-0.1388],[92,-0.1422],[94,-0.1411],[96,-0.1421],[98,-0.1411],[100,-0.1414],[102,-0.1414],[106,-0.1468],[108,-0.1473],[110,-0.1463],[112,-0.1435],[116,-0.1345],[120,-0.1358],[122,-0.1383],[124,-0.1387],[126,-0.1386],[128,-0.1371],[130,-0.1371],[132,-0.1383],[134,-0.1372],[136,-0.1373],[138,-0.138],[140,-0.1373],[142,-0.1361],[144,-0.1357],[148,-0.1324],[150,-0.1344],[152,-0.1332],[154,-0.1338],[156,-0.1338],[158,-0.1334],[160,-0.1313],[162,-0.1338],[164,-0.1315],[166,-0.1319],[168,-0.1331],[172,-0.1341],[176,-0.1347],[178,-0.1369],[180,-0.1369],[182,-0.1404],[184,-0.1427],[186,-0.1434],[188,-0.1453],[190,-0.1461],[194,-0.1454],[196,-0.1444],[198,-0.1456],[200,-0.1483],[202,-0.1492],[204,-0.1491],[208,-0.1528],[210,-0.1529],[212,-0.1561],[214,-0.1583],[218,-0.165],[220,-0.1667],[222,-0.1695],[224,-0.1715],[226,-0.1761],[228,-0.1765],[232,-0.1848],[234,-0.1879],[236,-0.1928],[238,-0.1964],[240,-0.2005],[242,-0.2026],[244,-0.2028],[246,-0.206],[248,-0.2078],[250,-0.2076],[252,-0.2088],[254,-0.2093],[258,-0.208],[260,-0.2089],[262,-0.2085],[264,-0.2077],[266,-0.2099],[268,-0.215],[270,-0.2185],[272,-0.2208],[274,-0.2297],[276,-0.2357],[278,-0.2307],[280,-0.2325],[282,-0.2322],[284,-0.2334],[288,-0.2187],[290,-0.2117],[292,-0.2079],[294,-0.2103],[296,-0.2086],[298,-0.2127],[302,-0.2197],[304,-0.2172],[308,-0.2176],[310,-0.208],[314,-0.2018],[316,-0.1957],[318,-0.1924],[320,-0.1851],[322,-0.1816],[324,-0.1751],[326,-0.1761],[328,-0.1758],[330,-0.1735],[332,-0.1674],[334,-0.1636],[336,-0.156],[338,-0.1545],[340,-0.1507],[344,-0.1412],[346,-0.1326],[348,-0.1227],[350,-0.1151],[352,-0.1122],[354,-0.1111],[356,-0.1063],[358,-0.1005],[360,-0.1006],[362,-0.0981],[364,-0.0942],[366,-0.0864],[368,-0.0801],[370,-0.0746],[372,-0.0623],[374,-0.0395],[376,-0.0353],[378,-0.0083],[380,0.027],[382,0.0593],[384,-0.0114],[386,-0.1183],[388,-0.2796],[390,-0.4059],[392,-0.6238],[394,-0.6916],[396,-0.6221],[398,-0.5062],[400,-0.4753],[402,-0.425],[404,-0.393],[406,-0.3163],[408,-0.2555],[410,-0.197],[412,-0.1587],[414,-0.1265],[416,-0.0914],[418,-0.0691],[420,-0.0608],[422,-0.0451],[424,-0.036],[426,-0.0232],[428,-0.0136],[430,-0.0123],[432,-0.0057],[434,0.0021],[436,0.0067],[438,0.0088],[442,0.0151],[444,0.018],[446,0.0139],[448,0.0153],[450,0.0104],[452,0.004],[454,-0.006],[456,-0.0196],[458,-0.0281],[460,-0.0397],[462,-0.0625],[464,-0.0838],[466,-0.1162],[468,-0.0977],[470,0.0554],[472,0.1467],[474,0.1289],[476,0.1285],[478,0.2624],[480,0.4735],[482,0.7439],[484,0.7949],[486,0.7897],[488,0.6941],[490,0.5877],[492,0.4852],[494,0.456],[496,0.4234],[500,0.3252],[502,0.2851],[504,0.2376],[506,0.2068],[508,0.1723],[510,0.1464],[512,0.1258],[514,0.1011],[516,0.0807],[518,0.0785],[520,0.059],[522,0.051],[524,0.0382],[526,-0.0096],[530,-0.0084],[532,-0.0025],[534,0.034],[536,0.0222],[538,0.0399],[540,0.0458],[542,0.0253],[544,0.0175],[546,0.0209],[548,0.0272],[550,0.0369],[552,0.0533],[554,0.0666],[556,0.0673],[558,0.067],[560,0.067],[562,0.0677],[564,0.0724],[566,0.0693],[568,0.0686],[570,0.0658],[572,0.0691],[574,0.069],[576,0.0681],[578,0.0649],[580,0.0639],[582,0.0717],[584,0.1441],[586,0.1902],[588,0.1884],[590,0.1972],[592,0.1702],[594,0.1637],[596,0.119],[598,0.0935],[600,0.0625],[602,0.0631],[604,0.0662],[606,0.0653],[608,0.0608],[612,0.0461],[614,0.0415],[616,0.0388],[618,0.0372],[620,0.0367],[622,0.0359],[624,0.0347],[626,0.035],[628,0.038],[630,0.0432],[632,0.0498],[634,0.0558],[636,0.0558],[638,0.057],[640,0.0561],[642,0.0515],[644,0.0571],[646,0.0568],[648,0.052],[650,0.05],[652,0.0505],[656,0.0484],[658,0.0505],[660,0.0519],[662,0.0506],[664,0.0507],[666,0.0535],[668,0.0571],[670,0.0579],[672,0.0576],[674,0.0474],[676,0.0454],[678,0.0439],[680,0.044],[682,0.0457],[684,0.0471],[686,0.0434],[690,0.047],[692,0.0489],[694,0.0476],[696,0.0501],[698,0.0651],[700,0.0669],[702,0.0645],[704,0.0614],[706,0.0607],[708,0.0602],[710,0.0593],[712,0.059],[714,0.0572],[716,0.0536],[718,0.0556],[720,0.0556]]
       }
       CTRL_L_eye.Y {
        type=MovieSceneScriptingFloatChannel
        default=0
        thinned_from=492
        keys [[-262,0.4323],[-260,0.3986],[-258,0.4048],[-256,0.4014],[-254,0.3808],[-252,0.3593],[-250,0.3499],[-248,0.3355],[-244,0.3374],[-242,0.3321],[-240,0.3254],[-234,0.3248],[-232,0.3223],[-230,0.3182],[-228,0.3073],[-226,0.2955],[-224,0.2786],[-222,0.2667],[-220,0.2649],[-218,0.2754],[-216,0.2826],[-214,0.2856],[-212,0.2861],[-210,0.2889],[-208,0.2774],[-206,0.2645],[-204,0.2565],[-202,0.2459],[-200,0.2337],[-198,0.2228],[-196,0.2096],[-194,0.1986],[-192,0.1908],[-190,0.1882],[-188,0.1777],[-186,0.177],[-184,0.1776],[-182,0.1847],[-180,0.1898],[-176,0.2154],[-174,0.2318],[-172,0.245],[-168,0.2683],[-166,0.2747],[-164,0.2834],[-160,0.2865],[-158,0.2873],[-156,0.2947],[-154,0.2993],[-152,0.2944],[-150,0.3016],[-148,0.2868],[-146,0.2753],[-144,0.2659],[-142,0.2585],[-140,0.2646],[-138,0.262],[-136,0.263],[-134,0.2662],[-132,0.2677],[-130,0.2721],[-128,0.2721],[-126,0.2801],[-124,0.2791],[-120,0.2802],[-118,0.2783],[-116,0.2838],[-114,0.2836],[-112,0.2826],[-110,0.2838],[-108,0.2814],[-106,0.2803],[-104,0.2819],[-102,0.2845],[-100,0.2911],[-98,0.2918],[-94,0.3205],[-92,0.3227],[-90,0.3283],[-88,0.3295],[-86,0.3345],[-84,0.3421],[-82,0.3444],[-80,0.3377],[-78,0.3272],[-76,0.3238],[-74,0.3154],[-72,0.3092],[-70,0.3052],[-68,0.3074],[-66,0.3043],[-64,0.3026],[-62,0.3036],[-60,0.3056],[-56,0.3069],[-54,0.3084],[-52,0.3079],[-50,0.3003],[-48,0.2959],[-46,0.1797],[-44,0.1231],[-42,0.2244],[-40,0.2526],[-38,0.2528],[-36,0.2582],[-34,0.2694],[-32,0.28],[-30,0.2814],[-28,0.2779],[-26,0.2782],[-24,0.2805],[-22,0.2733],[-20,0.2789],[-18,0.2837],[-16,0.5693],[-14,0.5715],[-12,0.5721],[-10,0.579],[-4,0.583],[-2,0.583],[0,0.5798],[2,0.5801],[4,0.5778],[6,0.579],[8,0.579],[14,0.5833],[18,0.5831],[22,0.5865],[26,0.5875],[28,0.5895],[34,0.5898],[36,0.5916],[40,0.5936],[42,0.5914],[44,0.5937],[48,0.5947],[50,0.5942],[52,0.5963],[58,0.5991],[60,0.6017],[62,0.6058],[70,0.6092],[72,0.6126],[76,0.6078],[78,0.6075],[80,0.6097],[82,0.6104],[84,0.6088],[88,0.6115],[90,0.6113],[94,0.6064],[96,0.5994],[98,0.5942],[100,0.5866],[102,0.571],[104,0.5608],[106,0.5349],[108,0.5317],[112,0.5095],[114,0.5034],[116,0.5027],[120,0.5038],[126,0.5014],[128,0.4959],[130,0.4959],[132,0.4916],[134,0.4911],[136,0.4926],[138,0.4928],[140,0.4982],[144,0.506],[146,0.5181],[148,0.525],[150,0.5294],[154,0.5451],[156,0.5454],[158,0.5495],[160,0.5549],[162,0.5583],[164,0.5592],[170,0.5667],[174,0.5678],[176,0.5712],[178,0.5732],[180,0.5713],[182,0.5717],[184,0.5688],[190,0.5688],[192,0.5659],[194,0.5654],[198,0.5616],[206,0.5656],[208,0.5684],[212,0.5693],[218,0.575],[220,0.5793],[222,0.5794],[228,0.5833],[230,0.5897],[236,0.5935],[240,0.5991],[242,0.6053],[244,0.6097],[246,0.6184],[248,0.6182],[250,0.6213],[252,0.6219],[256,0.6215],[260,0.6199],[262,0.6209],[264,0.6173],[266,0.6106],[268,0.6093],[270,0.6137],[272,0.6314],[274,0.6627],[276,0.7064],[278,0.7404],[280,0.7661],[282,0.7752],[284,0.7864],[286,0.7819],[290,0.7667],[292,0.769],[294,0.7604],[296,0.7663],[298,0.7604],[300,0.7634],[304,0.7565],[308,0.7535],[310,0.746],[312,0.7483],[316,0.7425],[318,0.7321],[320,0.7452],[322,0.7501],[324,0.7498],[326,0.7523],[328,0.7651],[330,0.7633],[332,0.7699],[334,0.7856],[336,0.8051],[340,0.7992],[344,0.7528],[346,0.7212],[348,0.694],[350,0.6687],[352,0.6512],[354,0.6455],[356,0.6296],[358,0.6275],[360,0.621],[362,0.6182],[366,0.616],[368,0.6116],[370,0.6032],[372,0.5895],[374,0.5526],[376,0.546],[378,0.5188],[380,0.4773],[382,0.4482],[384,0.4478],[386,0.4801],[388,0.6333],[390,0.6625],[392,0.6677],[394,0.6227],[398,0.4125],[400,0.3886],[402,0.34],[404,0.222],[406,0.2733],[408,0.2491],[410,0.2068],[412,0.1778],[414,0.1596],[416,0.1439],[418,0.132],[420,0.167],[422,0.1799],[424,0.1724],[426,0.1531],[428,0.1366],[430,0.126],[432,0.1188],[434,0.1131],[436,0.1151],[438,0.1159],[440,0.115],[442,0.1218],[444,0.1254],[448,0.1298],[450,0.1337],[452,0.1417],[454,0.1513],[456,0.1673],[458,0.1883],[460,0.2128],[462,0.2521],[464,0.2889],[466,0.354],[468,0.3739],[470,0.3339],[472,0.3121],[474,0.2172],[476,0.1057],[478,0.2004],[480,0.1694],[482,0.0538],[484,0.0311],[486,0.0169],[488,0.0555],[490,0.1047],[492,0.1644],[494,0.1685],[496,0.1805],[498,0.2104],[500,0.24],[502,0.2457],[504,0.2703],[506,0.3018],[508,0.3161],[510,0.3342],[512,0.342],[514,0.2681],[516,-0.0051],[518,-0.0419],[520,-0.4233],[522,-0.6592],[524,-0.7545],[526,-0.776],[528,-0.7743],[530,-0.7628],[532,-0.7532],[534,-0.4432],[536,-0.1928],[538,0.2882],[540,0.3805],[542,0.5936],[544,0.5561],[546,0.5125],[548,0.4338],[550,0.3958],[552,0.3982],[554,0.3977],[556,0.3874],[558,0.3851],[560,0.3818],[562,0.3842],[564,0.3849],[566,0.3876],[568,0.3876],[570,0.3801],[572,0.3874],[578,0.3892],[580,0.3886],[582,0.3777],[584,0.3503],[586,0.3333],[588,0.3439],[590,0.3607],[592,0.3601],[594,0.3568],[596,0.2858],[598,0.2364],[600,0.1579],[602,0.1531],[604,0.115],[606,0.1086],[608,0.0895],[610,0.0783],[612,0.0652],[614,0.0566],[616,0.0505],[618,0.0468],[620,0.0456],[622,0.0433],[624,0.0425],[626,0.0428],[628,0.0454],[630,0.0445],[632,0.0531],[634,0.0553],[636,0.062],[638,0.0636],[640,0.0668],[642,0.0604],[646,0.0644],[648,0.0598],[650,0.0603],[654,0.0689],[656,0.066],[658,0.0713],[660,0.0771],[662,0.0795],[664,0.0792],[666,0.0897],[668,0.0949],[670,0.0991],[672,0.1016],[674,0.101],[676,0.105],[678,0.1035],[680,0.1082],[682,0.1077],[684,0.1094],[686,0.1046],[688,0.1079],[690,0.1044],[692,0.1072],[694,0.1067],[696,0.109],[698,0.1158],[700,0.1186],[702,0.1204],[704,0.1227],[706,0.1262],[708,0.1303],[710,0.1324],[712,0.1322],[714,0.1342],[716,0.1391],[718,0.1434],[720,0.1434]]
       }
       CTRL_L_eye_blink {
        type=MovieSceneScriptingFloatChannel
        default=0
        thinned_from=492
        keys [[-262,-0.476],[-260,-0.5154],[-258,-0.5307],[-256,-0.5539],[-254,-0.5738],[-252,-0.5777],[-250,-0.5879],[-248,-0.5835],[-246,-0.5766],[-244,-0.5752],[-242,-0.5811],[-240,-0.5356],[-238,-0.5135],[-236,-0.5035],[-234,-0.5025],[-232,-0.4951],[-230,-0.4894],[-228,-0.492],[-226,-0.4975],[-224,-0.4947],[-222,-0.4935],[-220,-0.4986],[-218,-0.4907],[-216,-0.4839],[-212,-0.4832],[-210,-0.469],[-208,-0.4736],[-206,-0.4845],[-204,-0.4909],[-202,-0.5047],[-200,-0.5091],[-198,-0.4938],[-196,-0.4848],[-194,-0.4804],[-190,-0.4789],[-188,-0.4612],[-186,-0.4608],[-184,-0.462],[-180,-0.4434],[-178,-0.4266],[-176,-0.4215],[-174,-0.4194],[-170,-0.3701],[-168,-0.3607],[-166,-0.361],[-164,-0.3638],[-162,-0.3591],[-160,-0.3504],[-158,-0.3506],[-156,-0.3518],[-154,-0.3383],[-152,-0.3191],[-148,-0.319],[-146,-0.3179],[-144,-0.328],[-142,-0.3227],[-140,-0.3104],[-138,-0.3143],[-136,-0.3213],[-134,-0.322],[-132,-0.3201],[-130,-0.3172],[-128,-0.3236],[-126,-0.318],[-124,-0.3136],[-122,-0.3175],[-118,-0.3097],[-116,-0.3157],[-114,-0.3152],[-110,-0.3103],[-108,-0.3155],[-106,-0.314],[-104,-0.3109],[-102,-0.3033],[-100,-0.3061],[-98,-0.3064],[-96,-0.3119],[-94,-0.3099],[-90,-0.3105],[-88,-0.3144],[-86,-0.3299],[-84,-0.3428],[-82,-0.3424],[-78,-0.338],[-76,-0.337],[-74,-0.3425],[-72,-0.3313],[-70,-0.3237],[-68,-0.3192],[-66,-0.306],[-64,-0.3009],[-62,-0.3004],[-60,-0.2967],[-58,-0.2983],[-56,-0.2971],[-54,-0.2822],[-52,-0.2778],[-50,-0.2707],[-48,-0.2623],[-46,0.0789],[-44,0.7483],[-42,0.6325],[-40,0.3172],[-38,0.0607],[-36,-0.0662],[-34,-0.1803],[-32,-0.2798],[-30,-0.3374],[-28,-0.3725],[-26,-0.3794],[-24,-0.3839],[-22,-0.3811],[-20,-0.3739],[-18,-0.3847],[-16,-0.4295],[-14,-0.4263],[-12,-0.4283],[-10,-0.4316],[-8,-0.4361],[-6,-0.4386],[-4,-0.4376],[-2,-0.4355],[0,-0.4288],[2,-0.4247],[8,-0.4081],[10,-0.4011],[12,-0.3827],[14,-0.3816],[22,-0.3842],[24,-0.381],[26,-0.3769],[28,-0.3744],[32,-0.376],[34,-0.3781],[36,-0.3772],[38,-0.3727],[40,-0.3743],[42,-0.3744],[44,-0.3732],[46,-0.3739],[48,-0.3764],[50,-0.3779],[54,-0.3792],[56,-0.3777],[58,-0.3716],[60,-0.3685],[64,-0.3764],[66,-0.3743],[68,-0.371],[72,-0.3708],[74,-0.3688],[78,-0.3687],[80,-0.3705],[86,-0.3739],[88,-0.377],[90,-0.3782],[92,-0.3821],[94,-0.3834],[96,-0.3809],[98,-0.3825],[102,-0.3759],[104,-0.3762],[106,-0.3557],[108,-0.3521],[110,-0.3446],[112,-0.3414],[114,-0.3422],[116,-0.347],[118,-0.3454],[120,-0.3445],[124,-0.3334],[126,-0.3332],[128,-0.33],[130,-0.33],[132,-0.3241],[134,-0.3255],[136,-0.3353],[138,-0.3346],[140,-0.3283],[144,-0.3207],[148,-0.3161],[150,-0.3116],[152,-0.3097],[154,-0.3113],[158,-0.311],[160,-0.3146],[162,-0.3148],[164,-0.3169],[166,-0.3229],[168,-0.3222],[172,-0.3255],[176,-0.3267],[178,-0.3264],[184,-0.3353],[186,-0.3362],[188,-0.3382],[190,-0.3378],[192,-0.338],[196,-0.3411],[200,-0.3381],[206,-0.3381],[208,-0.34],[210,-0.3401],[212,-0.3428],[216,-0.3379],[218,-0.3409],[222,-0.3347],[224,-0.3381],[226,-0.3407],[228,-0.3447],[230,-0.3426],[232,-0.3419],[234,-0.3497],[236,-0.3518],[238,-0.3491],[240,-0.3444],[242,-0.3412],[244,-0.3409],[246,-0.3452],[248,-0.3515],[250,-0.353],[254,-0.3543],[256,-0.356],[258,-0.3586],[262,-0.3666],[264,-0.3718],[266,-0.3747],[268,-0.3746],[270,-0.3793],[272,-0.3934],[274,-0.4085],[276,-0.4199],[280,-0.4488],[282,-0.4514],[284,-0.4628],[286,-0.4718],[288,-0.488],[290,-0.4978],[292,-0.4975],[294,-0.5002],[296,-0.4918],[298,-0.4966],[300,-0.4959],[302,-0.4875],[306,-0.4768],[310,-0.4833],[312,-0.4908],[314,-0.4971],[316,-0.4936],[318,-0.4825],[322,-0.5018],[324,-0.4886],[328,-0.5065],[330,-0.493],[332,-0.4777],[334,-0.4717],[336,-0.4732],[338,-0.4656],[340,-0.4601],[342,-0.4609],[344,-0.474],[346,-0.5014],[348,-0.5167],[350,-0.5227],[352,-0.5251],[354,-0.5229],[356,-0.5173],[360,-0.5336],[362,-0.5278],[364,-0.5329],[366,-0.5348],[368,-0.5307],[370,-0.5219],[372,-0.5193],[374,-0.5349],[376,-0.5378],[378,-0.5459],[380,-0.5345],[382,-0.5087],[384,-0.1639],[386,0.4198],[388,0.5632],[390,0.4126],[392,0.0805],[394,-0.1088],[396,-0.2465],[398,-0.3122],[400,-0.3197],[402,-0.2996],[404,-0.2458],[406,-0.263],[408,-0.3266],[410,-0.3403],[412,-0.3481],[414,-0.3472],[418,-0.3489],[420,-0.3283],[422,-0.3335],[424,-0.3509],[426,-0.3797],[428,-0.4012],[430,-0.4103],[432,-0.41],[434,-0.4127],[436,-0.4173],[438,-0.4085],[440,-0.4038],[442,-0.4041],[444,-0.4011],[448,-0.4002],[450,-0.4021],[452,-0.4051],[454,-0.397],[456,-0.3897],[458,-0.3987],[460,-0.4027],[462,-0.3943],[464,-0.3814],[466,-0.3282],[468,-0.2128],[470,0.078],[472,0.5362],[474,0.7102],[476,0.7303],[478,0.6507],[480,0.4425],[482,0.1831],[484,0.1367],[486,-0.0165],[488,-0.0698],[490,-0.139],[492,-0.2134],[496,-0.3385],[498,-0.3943],[500,-0.4139],[502,-0.4241],[504,-0.4055],[506,-0.3928],[508,-0.4188],[510,-0.4374],[512,-0.4164],[514,-0.2628],[516,-0.116],[518,-0.1255],[520,-0.2707],[522,-0.4476],[524,-0.4851],[526,-0.4254],[528,-0.4189],[530,-0.3991],[532,-0.3628],[534,0.2196],[536,0.6807],[538,0.4488],[540,0.3843],[542,-0.0036],[544,-0.2122],[546,-0.3575],[548,-0.4603],[550,-0.4814],[552,-0.4767],[556,-0.4948],[558,-0.489],[560,-0.489],[562,-0.4983],[564,-0.4982],[566,-0.4929],[568,-0.4929],[570,-0.5081],[572,-0.5293],[574,-0.524],[576,-0.5138],[578,-0.5093],[580,-0.5158],[582,-0.4547],[584,-0.0499],[586,0.5423],[588,0.5229],[590,0.2416],[592,0.0317],[594,0.0002],[596,-0.069],[598,-0.1193],[600,-0.1744],[602,-0.1759],[604,-0.1873],[606,-0.1925],[608,-0.2098],[610,-0.2192],[612,-0.2104],[614,-0.1903],[616,-0.1692],[618,-0.1476],[620,-0.1407],[622,-0.1365],[624,-0.1365],[626,-0.136],[630,-0.1509],[632,-0.155],[634,-0.1575],[636,-0.1554],[638,-0.1504],[640,-0.1476],[642,-0.1461],[644,-0.1547],[646,-0.166],[648,-0.1641],[650,-0.1557],[652,-0.1543],[656,-0.1592],[658,-0.1611],[660,-0.1601],[662,-0.161],[664,-0.1592],[666,-0.1585],[668,-0.1619],[670,-0.1619],[672,-0.1601],[674,-0.1571],[676,-0.1571],[678,-0.1622],[680,-0.1685],[682,-0.1713],[684,-0.168],[686,-0.1699],[688,-0.1762],[690,-0.1743],[692,-0.174],[694,-0.1814],[696,-0.1844],[698,-0.1742],[700,-0.1731],[702,-0.1752],[704,-0.1666],[706,-0.1606],[708,-0.1622],[710,-0.1681],[712,-0.1696],[714,-0.1679],[716,-0.167],[718,-0.1639],[720,-0.1639]]
       }
       CTRL_R_eye.X {
        type=MovieSceneScriptingFloatChannel
        default=0
        thinned_from=492
        keys [[-262,0.0121],[-260,0.0063],[-258,0.0049],[-256,0.004],[-254,-0.0066],[-252,-0.0138],[-250,-0.0214],[-248,-0.0269],[-246,-0.0315],[-244,-0.0327],[-242,-0.0279],[-240,-0.0323],[-238,-0.0275],[-236,-0.016],[-234,-0.0152],[-232,-0.0107],[-230,-0.0064],[-228,0.001],[-226,0.0045],[-224,0.006],[-222,0.0085],[-220,0.0122],[-218,0.0123],[-216,0.0107],[-214,0.0094],[-212,0.0089],[-210,0],[-208,-0.0044],[-206,-0.0084],[-204,-0.0083],[-202,-0.0077],[-200,-0.0139],[-198,-0.016],[-196,-0.0196],[-194,-0.0237],[-192,-0.0271],[-190,-0.0282],[-188,-0.0316],[-186,-0.0317],[-184,-0.0333],[-182,-0.0352],[-180,-0.0377],[-176,-0.0413],[-174,-0.0439],[-172,-0.0472],[-170,-0.0456],[-168,-0.0398],[-166,-0.0374],[-164,-0.0313],[-162,-0.0261],[-160,-0.0224],[-158,-0.0215],[-154,-0.0055],[-152,-0.007],[-150,-0.0015],[-148,-0.0047],[-146,-0.0097],[-144,-0.0116],[-142,-0.0139],[-140,-0.0237],[-138,-0.026],[-136,-0.0262],[-134,-0.0259],[-132,-0.0227],[-130,-0.0149],[-128,-0.015],[-126,-0.0103],[-124,-0.0118],[-122,-0.0139],[-120,-0.0142],[-118,-0.0109],[-116,-0.0098],[-114,-0.0101],[-110,-0.0136],[-108,-0.0141],[-106,-0.0173],[-104,-0.0191],[-102,-0.0229],[-100,-0.0317],[-98,-0.0327],[-96,-0.0303],[-94,-0.0268],[-92,-0.0273],[-90,-0.0288],[-88,-0.0291],[-86,-0.0288],[-84,-0.025],[-82,-0.0251],[-80,-0.0285],[-78,-0.0275],[-76,-0.0264],[-74,-0.017],[-72,-0.0177],[-70,-0.0155],[-68,-0.0095],[-66,-0.005],[-64,-0.0016],[-62,-0.0006],[-60,0.0029],[-58,0.0056],[-56,0.0092],[-54,0.0069],[-52,0.005],[-50,0.0034],[-48,0.0031],[-46,0.0019],[-44,0.1088],[-42,0.1056],[-40,0.0228],[-38,-0.0069],[-36,-0.01],[-34,-0.0048],[-32,-0.0035],[-30,-0.0047],[-28,-0.0089],[-26,-0.0082],[-24,-0.0071],[-22,-0.0098],[-20,-0.0095],[-18,-0.0133],[-16,-0.1469],[-12,-0.1444],[-10,-0.148],[-8,-0.1474],[-6,-0.1477],[-4,-0.1469],[-2,-0.1453],[0,-0.1444],[2,-0.144],[4,-0.1442],[6,-0.1432],[8,-0.143],[10,-0.1415],[12,-0.1412],[22,-0.1429],[24,-0.1429],[26,-0.1419],[28,-0.1421],[30,-0.1441],[32,-0.1454],[34,-0.1443],[36,-0.1443],[40,-0.1453],[42,-0.1473],[44,-0.1461],[48,-0.1453],[50,-0.1459],[54,-0.1448],[58,-0.1459],[60,-0.1475],[62,-0.148],[64,-0.146],[66,-0.1466],[68,-0.1488],[70,-0.1485],[72,-0.147],[74,-0.1475],[76,-0.1484],[78,-0.1486],[80,-0.1495],[82,-0.1493],[84,-0.1505],[86,-0.1522],[88,-0.1534],[90,-0.1535],[92,-0.1567],[94,-0.1562],[96,-0.1573],[98,-0.1572],[100,-0.1583],[102,-0.1589],[106,-0.1667],[108,-0.1677],[110,-0.1675],[112,-0.1647],[116,-0.1567],[118,-0.1576],[120,-0.1582],[122,-0.1603],[124,-0.1614],[126,-0.1613],[128,-0.1603],[130,-0.1603],[132,-0.1622],[134,-0.1613],[136,-0.1613],[138,-0.1626],[142,-0.1602],[144,-0.1593],[146,-0.1576],[148,-0.1554],[150,-0.1571],[152,-0.1549],[154,-0.1554],[158,-0.1553],[160,-0.1527],[162,-0.156],[164,-0.1527],[166,-0.1532],[168,-0.1544],[170,-0.1547],[172,-0.1555],[174,-0.1555],[176,-0.1566],[178,-0.1581],[180,-0.1583],[182,-0.1615],[184,-0.1643],[186,-0.1648],[188,-0.1661],[190,-0.1668],[192,-0.1663],[194,-0.1663],[196,-0.1651],[198,-0.1664],[200,-0.1697],[202,-0.1703],[204,-0.1701],[206,-0.172],[208,-0.1743],[210,-0.1744],[214,-0.1797],[216,-0.1828],[220,-0.187],[222,-0.1895],[224,-0.191],[226,-0.1954],[228,-0.1964],[230,-0.2012],[232,-0.2053],[234,-0.2087],[236,-0.2134],[238,-0.2166],[240,-0.2209],[242,-0.2235],[244,-0.2229],[248,-0.2264],[250,-0.2259],[252,-0.2276],[254,-0.2284],[258,-0.2274],[260,-0.2279],[262,-0.2275],[264,-0.2266],[266,-0.229],[268,-0.2347],[270,-0.2384],[272,-0.2414],[274,-0.2495],[276,-0.2556],[278,-0.2475],[280,-0.2477],[282,-0.2463],[284,-0.2473],[286,-0.2396],[288,-0.2349],[290,-0.2294],[292,-0.2261],[294,-0.2278],[296,-0.2244],[298,-0.2303],[300,-0.2342],[302,-0.2371],[304,-0.2343],[306,-0.2371],[308,-0.2375],[310,-0.2283],[314,-0.2219],[316,-0.2147],[318,-0.2097],[320,-0.2019],[322,-0.1978],[324,-0.1903],[326,-0.1933],[328,-0.1905],[330,-0.1884],[332,-0.1811],[334,-0.1773],[336,-0.1694],[338,-0.1701],[340,-0.1662],[342,-0.1599],[344,-0.155],[346,-0.1473],[348,-0.138],[350,-0.1304],[354,-0.1277],[356,-0.1221],[358,-0.116],[360,-0.115],[362,-0.1128],[364,-0.1083],[368,-0.0955],[370,-0.0895],[372,-0.0776],[374,-0.0559],[376,-0.0518],[378,-0.0235],[380,0.011],[382,0.0434],[384,-0.0432],[386,-0.1448],[388,-0.3416],[390,-0.4562],[392,-0.6439],[394,-0.7003],[396,-0.6318],[398,-0.5203],[400,-0.4907],[402,-0.4433],[404,-0.41],[406,-0.3309],[408,-0.2708],[410,-0.2157],[412,-0.1798],[414,-0.1511],[416,-0.117],[418,-0.0948],[420,-0.084],[422,-0.0674],[424,-0.0581],[426,-0.0463],[428,-0.0375],[430,-0.0362],[432,-0.0305],[434,-0.0228],[436,-0.0178],[440,-0.0116],[442,-0.0077],[444,-0.0049],[446,-0.0088],[448,-0.0075],[450,-0.0122],[452,-0.0185],[454,-0.0281],[456,-0.0403],[458,-0.048],[460,-0.0591],[462,-0.0817],[464,-0.1016],[466,-0.1363],[468,-0.121],[470,0.0249],[472,0.146],[474,0.198],[476,0.2198],[478,0.3765],[480,0.564],[482,0.7567],[484,0.7923],[486,0.7673],[488,0.6703],[492,0.4601],[494,0.4324],[496,0.3965],[498,0.344],[500,0.2947],[502,0.2533],[504,0.2077],[506,0.1779],[508,0.1425],[510,0.1183],[512,0.0986],[514,0.0694],[516,0.0429],[518,0.041],[520,0.0351],[522,0.0675],[524,0.0921],[526,0.0402],[528,0.0438],[530,0.0446],[532,0.0382],[534,0.0765],[536,0.0746],[538,0.0308],[540,0.0234],[542,-0.0045],[544,-0.0129],[546,-0.0089],[548,-0.002],[550,0.0087],[552,0.0257],[554,0.0382],[556,0.0391],[558,0.0391],[560,0.0402],[562,0.0408],[564,0.0451],[566,0.0426],[568,0.0419],[570,0.0386],[572,0.0409],[574,0.0411],[576,0.04],[578,0.0354],[580,0.0336],[582,0.0414],[584,0.1119],[586,0.1908],[588,0.1731],[590,0.1527],[592,0.1276],[594,0.1219],[596,0.0805],[598,0.0564],[600,0.0254],[602,0.026],[604,0.0281],[606,0.0268],[608,0.0214],[610,0.0141],[612,0.0059],[614,0.0002],[616,-0.0028],[618,-0.0054],[620,-0.0063],[622,-0.0064],[624,-0.007],[626,-0.0066],[628,-0.0041],[630,0.0011],[632,0.0079],[634,0.0145],[636,0.014],[638,0.0151],[640,0.0141],[642,0.009],[644,0.0155],[646,0.0169],[648,0.0114],[650,0.0091],[652,0.0109],[654,0.0096],[656,0.0104],[658,0.0118],[660,0.0124],[662,0.012],[664,0.0132],[666,0.0175],[668,0.021],[670,0.0218],[672,0.0221],[674,0.0121],[676,0.0109],[678,0.0084],[680,0.0087],[682,0.0111],[684,0.0114],[686,0.0086],[688,0.0109],[692,0.0131],[694,0.0116],[696,0.0145],[698,0.0281],[700,0.0303],[702,0.0268],[704,0.0247],[706,0.0243],[708,0.0242],[710,0.0232],[712,0.0229],[714,0.0211],[716,0.0182],[718,0.0203],[720,0.0203]]
       }
       CTRL_R_eye.Y {
        type=MovieSceneScriptingFloatChannel
        default=0
        thinned_from=492
        keys [[-262,0.431],[-260,0.3968],[-258,0.4029],[-256,0.3995],[-252,0.3584],[-250,0.3493],[-248,0.3346],[-244,0.3364],[-242,0.3311],[-240,0.323],[-238,0.3224],[-236,0.3235],[-234,0.3232],[-232,0.3204],[-230,0.3165],[-228,0.3058],[-226,0.2937],[-224,0.277],[-222,0.2656],[-220,0.2638],[-218,0.274],[-216,0.2817],[-214,0.2842],[-212,0.2846],[-210,0.2877],[-208,0.2757],[-206,0.2626],[-204,0.2546],[-202,0.244],[-200,0.2318],[-198,0.2213],[-196,0.2078],[-194,0.1971],[-192,0.1887],[-190,0.1858],[-188,0.1759],[-186,0.1753],[-184,0.1763],[-182,0.1824],[-180,0.1879],[-178,0.2013],[-176,0.2134],[-174,0.2294],[-172,0.2426],[-168,0.2658],[-166,0.272],[-164,0.2805],[-158,0.2828],[-156,0.29],[-154,0.2962],[-152,0.2925],[-150,0.3003],[-148,0.2864],[-146,0.2752],[-144,0.2659],[-142,0.2591],[-140,0.2642],[-138,0.2632],[-136,0.2635],[-134,0.267],[-132,0.2687],[-130,0.2725],[-128,0.2728],[-126,0.2804],[-124,0.2799],[-120,0.2815],[-118,0.2791],[-116,0.2855],[-114,0.2854],[-112,0.2845],[-110,0.2854],[-108,0.2834],[-106,0.2829],[-102,0.2861],[-100,0.2928],[-98,0.2935],[-96,0.3072],[-94,0.3215],[-92,0.3235],[-90,0.3291],[-88,0.3301],[-86,0.3345],[-84,0.3411],[-82,0.3428],[-80,0.3351],[-78,0.3241],[-76,0.3208],[-74,0.3129],[-72,0.3071],[-70,0.3046],[-68,0.3065],[-66,0.304],[-64,0.3025],[-62,0.3038],[-60,0.3067],[-56,0.3079],[-54,0.3094],[-52,0.3088],[-50,0.3017],[-48,0.2975],[-46,0.181],[-44,0.1214],[-42,0.2219],[-40,0.2511],[-38,0.2517],[-36,0.2572],[-34,0.269],[-32,0.2799],[-30,0.2814],[-28,0.2772],[-26,0.2773],[-24,0.2796],[-22,0.2737],[-18,0.283],[-16,0.5693],[-14,0.5718],[-12,0.5724],[-10,0.5792],[-8,0.5799],[-6,0.5824],[-2,0.5831],[0,0.5799],[2,0.58],[4,0.5779],[6,0.5792],[8,0.5789],[14,0.583],[18,0.5826],[22,0.5861],[26,0.5871],[28,0.5889],[34,0.5894],[40,0.5933],[42,0.5911],[44,0.5934],[50,0.5936],[52,0.5956],[58,0.5987],[60,0.6012],[62,0.6054],[64,0.607],[70,0.6089],[72,0.6123],[76,0.6074],[78,0.607],[80,0.6093],[82,0.6098],[84,0.6083],[88,0.6113],[90,0.611],[94,0.6064],[96,0.5992],[98,0.5943],[100,0.587],[102,0.5712],[104,0.5608],[106,0.5349],[108,0.5318],[112,0.5094],[114,0.5034],[116,0.5027],[120,0.5035],[126,0.5014],[128,0.4957],[130,0.4957],[132,0.4914],[134,0.4911],[136,0.4927],[138,0.493],[140,0.4984],[144,0.5059],[146,0.5183],[148,0.5249],[150,0.5293],[154,0.5456],[156,0.5459],[158,0.5495],[160,0.5548],[162,0.5581],[164,0.5592],[170,0.5668],[174,0.5683],[176,0.5718],[178,0.5733],[180,0.5717],[182,0.5722],[184,0.5693],[190,0.569],[192,0.5661],[194,0.5657],[198,0.5618],[200,0.5626],[202,0.5645],[206,0.566],[208,0.5689],[214,0.571],[218,0.5755],[220,0.5796],[224,0.581],[228,0.5839],[230,0.5903],[236,0.5938],[240,0.5993],[242,0.6055],[244,0.6099],[246,0.6186],[248,0.6184],[250,0.6217],[252,0.6224],[254,0.6212],[256,0.6219],[260,0.6207],[262,0.6214],[264,0.6178],[266,0.6115],[268,0.6099],[270,0.6141],[272,0.6316],[274,0.6632],[276,0.7076],[278,0.7412],[280,0.7664],[282,0.7756],[284,0.7867],[286,0.782],[290,0.7679],[292,0.7686],[294,0.7603],[296,0.767],[298,0.7613],[300,0.7644],[304,0.7566],[308,0.7535],[310,0.7466],[312,0.7492],[316,0.7443],[318,0.7341],[320,0.7465],[322,0.7512],[324,0.7508],[326,0.7532],[328,0.7654],[330,0.764],[332,0.7708],[334,0.7866],[336,0.8061],[338,0.8043],[340,0.8004],[344,0.7533],[346,0.7213],[350,0.6689],[352,0.6519],[354,0.6462],[356,0.6298],[358,0.6276],[360,0.621],[362,0.6174],[366,0.6158],[368,0.612],[370,0.6035],[372,0.5896],[374,0.5526],[376,0.546],[378,0.5188],[380,0.4772],[382,0.4476],[384,0.4462],[386,0.4776],[388,0.6293],[390,0.6585],[392,0.6637],[394,0.6188],[398,0.4098],[400,0.3858],[402,0.3378],[404,0.2194],[406,0.2711],[408,0.2478],[410,0.2043],[412,0.1753],[414,0.158],[416,0.1421],[418,0.1309],[420,0.1666],[422,0.1791],[424,0.1715],[426,0.1522],[428,0.135],[430,0.1253],[432,0.118],[434,0.1122],[436,0.1141],[438,0.1156],[440,0.1149],[442,0.1211],[444,0.1252],[446,0.1276],[448,0.1304],[450,0.1345],[452,0.142],[454,0.1514],[456,0.1666],[458,0.1881],[460,0.2128],[462,0.2521],[464,0.2883],[466,0.3531],[468,0.3729],[470,0.3321],[472,0.3086],[474,0.213],[476,0.1011],[478,0.195],[480,0.1649],[482,0.0508],[484,0.0284],[486,0.014],[488,0.0526],[490,0.1018],[492,0.1619],[494,0.1652],[496,0.1774],[498,0.2075],[500,0.2371],[502,0.2434],[504,0.2675],[506,0.2989],[508,0.3126],[510,0.3311],[512,0.3396],[514,0.2652],[516,-0.0079],[518,-0.0448],[520,-0.4277],[522,-0.6653],[524,-0.7612],[526,-0.7828],[528,-0.7813],[532,-0.7599],[534,-0.4493],[536,-0.1982],[538,0.285],[540,0.3777],[542,0.5907],[544,0.5545],[546,0.5107],[548,0.4317],[550,0.3933],[552,0.3961],[554,0.3957],[556,0.3854],[558,0.3835],[560,0.3802],[562,0.3819],[564,0.3827],[566,0.3855],[568,0.3856],[570,0.3781],[572,0.386],[578,0.3873],[580,0.3867],[582,0.3757],[584,0.3474],[586,0.3289],[588,0.3396],[590,0.3578],[592,0.3573],[594,0.354],[596,0.2839],[598,0.235],[600,0.1566],[602,0.1519],[604,0.1147],[606,0.1086],[608,0.0902],[610,0.0792],[612,0.0662],[614,0.0576],[616,0.0519],[618,0.0481],[620,0.0468],[622,0.0435],[624,0.0426],[626,0.0429],[628,0.0455],[630,0.0458],[632,0.0537],[634,0.0561],[636,0.0623],[638,0.0637],[640,0.0668],[642,0.0604],[644,0.0629],[646,0.0645],[648,0.0598],[650,0.0604],[654,0.0689],[656,0.066],[658,0.0712],[660,0.0758],[662,0.079],[664,0.0791],[666,0.0891],[668,0.0947],[670,0.099],[672,0.1016],[674,0.101],[676,0.1057],[678,0.1037],[680,0.1083],[682,0.1078],[684,0.1087],[686,0.1044],[688,0.1078],[690,0.1044],[692,0.1071],[694,0.1061],[696,0.1087],[698,0.1152],[700,0.1183],[702,0.1204],[704,0.1227],[706,0.1254],[708,0.1294],[710,0.1315],[712,0.1313],[714,0.134],[716,0.139],[718,0.1427],[720,0.1427]]
       }
       CTRL_R_eye_blink {
        type=MovieSceneScriptingFloatChannel
        default=0
        thinned_from=492
        keys [[-262,-0.6387],[-260,-0.6681],[-258,-0.6698],[-256,-0.6849],[-254,-0.7055],[-250,-0.7129],[-246,-0.7014],[-244,-0.7005],[-242,-0.7046],[-240,-0.6694],[-238,-0.6555],[-236,-0.6464],[-234,-0.6455],[-232,-0.6402],[-230,-0.6379],[-228,-0.6381],[-226,-0.6449],[-224,-0.6496],[-222,-0.6427],[-220,-0.6395],[-218,-0.6425],[-216,-0.637],[-214,-0.6222],[-212,-0.6201],[-208,-0.6179],[-206,-0.6129],[-204,-0.6148],[-202,-0.6233],[-200,-0.627],[-198,-0.6039],[-196,-0.5937],[-194,-0.581],[-192,-0.5724],[-190,-0.5701],[-188,-0.5548],[-186,-0.5528],[-184,-0.5431],[-182,-0.5398],[-178,-0.5166],[-176,-0.5205],[-174,-0.5297],[-172,-0.5226],[-170,-0.5066],[-168,-0.4862],[-166,-0.4843],[-162,-0.4863],[-160,-0.4805],[-156,-0.4806],[-154,-0.4843],[-152,-0.471],[-150,-0.4722],[-148,-0.4759],[-146,-0.4732],[-144,-0.4681],[-140,-0.4616],[-138,-0.459],[-136,-0.4623],[-134,-0.4616],[-132,-0.4565],[-130,-0.4551],[-128,-0.4616],[-126,-0.4695],[-124,-0.4699],[-122,-0.4684],[-120,-0.4636],[-118,-0.462],[-116,-0.4756],[-114,-0.4727],[-112,-0.4637],[-110,-0.4563],[-108,-0.4543],[-106,-0.4456],[-104,-0.4402],[-102,-0.4499],[-100,-0.451],[-98,-0.4509],[-96,-0.4602],[-94,-0.4655],[-90,-0.4649],[-88,-0.4684],[-86,-0.4822],[-84,-0.4932],[-82,-0.4949],[-80,-0.4888],[-78,-0.4704],[-76,-0.4648],[-74,-0.4838],[-72,-0.4711],[-70,-0.4604],[-68,-0.4595],[-66,-0.4565],[-64,-0.4544],[-58,-0.4574],[-56,-0.457],[-54,-0.4505],[-52,-0.4543],[-50,-0.4392],[-48,-0.4252],[-46,0.0037],[-44,0.8021],[-42,0.6935],[-40,0.3058],[-38,0.0604],[-36,-0.0656],[-34,-0.2012],[-32,-0.3268],[-30,-0.4085],[-28,-0.454],[-26,-0.4519],[-24,-0.4587],[-22,-0.4585],[-20,-0.4607],[-18,-0.4846],[-16,-0.5323],[-14,-0.5258],[-12,-0.5301],[-8,-0.5362],[-6,-0.5441],[-2,-0.5457],[0,-0.5396],[2,-0.5292],[4,-0.5164],[8,-0.5123],[10,-0.508],[12,-0.4992],[14,-0.5013],[16,-0.5004],[20,-0.4931],[26,-0.4939],[34,-0.4866],[36,-0.4818],[38,-0.4811],[40,-0.4852],[42,-0.4824],[46,-0.4836],[50,-0.4862],[52,-0.4909],[56,-0.4917],[58,-0.4912],[60,-0.4953],[62,-0.5032],[64,-0.4985],[66,-0.4955],[68,-0.4979],[72,-0.4972],[74,-0.4903],[76,-0.4933],[78,-0.4983],[80,-0.4968],[82,-0.4988],[84,-0.504],[86,-0.5032],[90,-0.5049],[92,-0.5042],[94,-0.5009],[98,-0.4972],[100,-0.4945],[102,-0.5066],[104,-0.5092],[106,-0.4825],[108,-0.4817],[110,-0.4777],[112,-0.4723],[114,-0.4692],[118,-0.4719],[122,-0.4726],[126,-0.471],[128,-0.4642],[130,-0.4642],[132,-0.4579],[134,-0.4552],[136,-0.4569],[138,-0.4557],[140,-0.4506],[142,-0.4492],[144,-0.4464],[148,-0.4478],[150,-0.4419],[152,-0.4413],[154,-0.448],[156,-0.448],[158,-0.4493],[160,-0.4544],[166,-0.4612],[168,-0.465],[170,-0.4706],[172,-0.4712],[174,-0.4756],[176,-0.4783],[178,-0.4791],[180,-0.4827],[182,-0.4847],[184,-0.4905],[186,-0.4921],[188,-0.4957],[190,-0.4946],[192,-0.4919],[194,-0.4904],[196,-0.4913],[198,-0.4906],[200,-0.4878],[202,-0.4902],[204,-0.4916],[206,-0.4899],[210,-0.4903],[212,-0.4936],[214,-0.4918],[216,-0.4869],[218,-0.4856],[222,-0.4871],[228,-0.4935],[230,-0.495],[232,-0.4886],[236,-0.4837],[238,-0.4727],[240,-0.4631],[242,-0.4611],[244,-0.462],[248,-0.4731],[252,-0.474],[254,-0.4754],[260,-0.4863],[262,-0.4886],[264,-0.4944],[266,-0.4942],[268,-0.492],[272,-0.5029],[274,-0.5144],[276,-0.5238],[278,-0.5307],[280,-0.5488],[282,-0.5594],[284,-0.5714],[286,-0.57],[288,-0.5708],[290,-0.5766],[292,-0.5973],[294,-0.5961],[296,-0.5887],[298,-0.6021],[300,-0.6042],[302,-0.5911],[304,-0.5876],[306,-0.5896],[308,-0.5976],[310,-0.5914],[312,-0.5967],[314,-0.6157],[316,-0.6172],[318,-0.598],[320,-0.5932],[322,-0.6104],[324,-0.6026],[326,-0.6073],[328,-0.6158],[330,-0.6029],[334,-0.6147],[336,-0.6164],[338,-0.6114],[340,-0.6092],[342,-0.6178],[344,-0.6366],[346,-0.6504],[348,-0.6598],[350,-0.6753],[354,-0.6767],[356,-0.6796],[358,-0.6851],[360,-0.6855],[362,-0.6844],[366,-0.6918],[370,-0.6943],[372,-0.7044],[374,-0.7113],[376,-0.712],[378,-0.7072],[380,-0.6953],[382,-0.6748],[384,-0.2838],[386,0.2818],[388,0.2892],[390,0.1503],[392,-0.087],[394,-0.2764],[396,-0.4593],[398,-0.5305],[400,-0.5391],[402,-0.4578],[404,-0.4066],[406,-0.4206],[408,-0.4062],[410,-0.3903],[412,-0.3801],[416,-0.3733],[418,-0.3584],[420,-0.3417],[422,-0.3674],[424,-0.3851],[426,-0.404],[428,-0.4203],[430,-0.4347],[434,-0.4329],[436,-0.4385],[438,-0.4394],[440,-0.4442],[442,-0.4474],[444,-0.4471],[448,-0.4491],[450,-0.4524],[452,-0.4513],[456,-0.429],[458,-0.4262],[460,-0.4209],[462,-0.4112],[464,-0.3977],[466,-0.3393],[468,-0.2456],[470,-0.0189],[472,0.4067],[474,0.6776],[476,0.7659],[478,0.5225],[480,0.2077],[482,-0.0153],[484,-0.0501],[486,-0.2506],[488,-0.443],[490,-0.5206],[492,-0.5698],[494,-0.634],[496,-0.6902],[498,-0.7101],[500,-0.6941],[502,-0.6862],[506,-0.6412],[510,-0.6579],[512,-0.6351],[514,-0.4223],[516,-0.1484],[518,-0.1535],[520,-0.2546],[522,-0.3915],[524,-0.3741],[526,-0.2957],[528,-0.2969],[530,-0.2774],[532,-0.2315],[534,0.3021],[536,0.7042],[538,0.3336],[540,0.2429],[542,-0.0896],[544,-0.3624],[546,-0.5081],[548,-0.6169],[550,-0.6451],[552,-0.6506],[554,-0.6654],[556,-0.6861],[558,-0.6895],[560,-0.6878],[562,-0.6951],[564,-0.6995],[566,-0.6965],[568,-0.6963],[570,-0.701],[572,-0.7242],[574,-0.7223],[576,-0.7161],[578,-0.7059],[580,-0.703],[582,-0.6276],[584,-0.1796],[586,0.4814],[588,0.4183],[590,0.0977],[592,-0.0395],[594,-0.0605],[596,-0.1899],[598,-0.2484],[600,-0.2936],[602,-0.2953],[604,-0.3076],[606,-0.3127],[608,-0.329],[610,-0.3355],[612,-0.322],[614,-0.2951],[616,-0.2611],[618,-0.2278],[620,-0.2173],[622,-0.2018],[624,-0.1992],[626,-0.199],[628,-0.2087],[630,-0.2192],[632,-0.2291],[634,-0.2361],[636,-0.242],[638,-0.2429],[640,-0.2422],[642,-0.2304],[644,-0.2385],[646,-0.2583],[648,-0.2583],[650,-0.2523],[652,-0.2546],[654,-0.2607],[658,-0.2607],[660,-0.2655],[662,-0.2681],[664,-0.2649],[666,-0.2682],[668,-0.2784],[670,-0.2814],[672,-0.2821],[674,-0.2754],[676,-0.2814],[678,-0.2976],[680,-0.3009],[682,-0.2967],[684,-0.2938],[686,-0.2956],[688,-0.2994],[690,-0.3005],[692,-0.3035],[694,-0.3079],[696,-0.3088],[698,-0.3124],[700,-0.3211],[702,-0.3305],[704,-0.3287],[706,-0.3257],[708,-0.3307],[710,-0.3365],[712,-0.3374],[714,-0.3364],[716,-0.3369],[718,-0.3339],[720,-0.3339]]
       }
       CTRL_C_eye.X {type=MovieSceneScriptingFloatChannel, default=0.6, const [0,491]}
       CTRL_C_eye.Y {type=MovieSceneScriptingFloatChannel, default=-0.65, const [0,492]}
       CTRL_C_eyesConvergence {type=MovieSceneScriptingFloatChannel, default=0, const [0,492]}
       CTRL_L_eye_cheekRaise {
        type=MovieSceneScriptingFloatChannel
        default=0
        thinned_from=492
        keys [[-262,0.0032],[-258,0.0023],[-256,0.0015],[-252,0.0007],[-250,0.0008],[-248,0.0012],[-244,0.0011],[-242,0.0007],[-240,0.0012],[-234,0.0014],[-230,0.0021],[-218,0.0012],[-216,0.0014],[-214,0.0019],[-210,0.0022],[-208,0.0021],[-206,0.0019],[-198,0.0012],[-188,0.0019],[-186,0.0019],[-182,0.0013],[-180,0.0014],[-172,0.0031],[-166,0.0036],[-156,0.0035],[-146,0.0028],[-132,0.0033],[-126,0.003],[-108,0.0034],[-102,0.0038],[-86,0.0038],[-80,0.0033],[-70,0.0034],[-62,0.0033],[-60,0.0043],[-58,0.0043],[-56,0.0046],[-52,0.0068],[-50,0.0071],[-48,0.0071],[-46,0.0054],[-42,0.0033],[-40,0.0035],[-38,0.0033],[-30,0.0034],[-24,0.0037],[-14,0.0037],[-12,0.0038],[-6,0.0037],[-2,0.0039],[8,0.0037],[10,0.0039],[12,0.0096],[14,0.0126],[16,0.0133],[18,0.0147],[22,0.015],[28,0.0124],[30,0.0121],[34,0.0118],[36,0.0105],[38,0.0086],[40,0.0082],[44,0.0099],[46,0.0099],[50,0.0096],[54,0.0115],[56,0.0123],[58,0.0127],[70,0.0122],[72,0.0117],[76,0.0117],[78,0.0122],[80,0.0123],[82,0.012],[84,0.0115],[86,0.0112],[88,0.0106],[90,0.0104],[94,0.0084],[96,0.0069],[98,0.006],[100,0.0058],[102,0.0061],[104,0.0056],[106,0.0042],[108,0.0048],[110,0.0061],[112,0.0067],[114,0.0068],[116,0.0065],[120,0.0064],[122,0.0058],[124,0.0068],[126,0.007],[128,0.0097],[132,0.0095],[134,0.0091],[136,0.0075],[138,0.0091],[140,0.0094],[142,0.0115],[144,0.0138],[146,0.0181],[148,0.0191],[150,0.021],[154,0.029],[156,0.029],[158,0.0302],[160,0.0327],[162,0.0317],[164,0.0335],[166,0.0342],[168,0.0359],[170,0.035],[172,0.0336],[176,0.0363],[178,0.0383],[180,0.0372],[182,0.0384],[184,0.0361],[186,0.0363],[190,0.0375],[192,0.0362],[194,0.0334],[196,0.0322],[198,0.0336],[200,0.0338],[202,0.0313],[204,0.0306],[206,0.0288],[208,0.0298],[210,0.0299],[212,0.0264],[216,0.0247],[218,0.0273],[220,0.0256],[222,0.0237],[224,0.0226],[226,0.0223],[230,0.0246],[232,0.0267],[234,0.0267],[236,0.0253],[238,0.0231],[240,0.0223],[242,0.0262],[244,0.0311],[246,0.0335],[248,0.0326],[250,0.032],[254,0.0336],[256,0.0336],[258,0.0345],[260,0.0357],[264,0.0361],[266,0.0371],[268,0.0385],[270,0.0361],[272,0.0297],[274,0.0199],[276,0.0206],[278,0.0247],[280,0.032],[282,0.0319],[284,0.0293],[286,0.0285],[288,0.0284],[290,0.0279],[292,0.0291],[296,0.0303],[298,0.0291],[300,0.031],[302,0.0332],[304,0.0349],[306,0.0343],[310,0.0315],[312,0.0309],[314,0.0316],[318,0.0313],[320,0.0299],[322,0.0306],[324,0.0292],[326,0.0285],[328,0.0272],[330,0.0255],[332,0.0224],[334,0.019],[336,0.0165],[338,0.0149],[340,0.0156],[342,0.0164],[344,0.0162],[346,0.0141],[350,0.0073],[352,0.0046],[356,0.0043],[358,0.0035],[360,0.0032],[362,0.0024],[364,0.0023],[368,0.0017],[374,0.0024],[376,0.0024],[382,0.0018],[384,0.0018],[386,0.0025],[388,0.0039],[390,0.004],[392,0.0037],[394,0.003],[396,0.0026],[400,0.0029],[402,0.0034],[404,0.0045],[408,0.0047],[410,0.0035],[416,0.0035],[420,0.0033],[424,0.0033],[428,0.0028],[438,0.0028],[442,0.0032],[456,0.0031],[468,0.0028],[470,0.0024],[472,0.0025],[476,0.0033],[480,0.0021],[486,0.0019],[488,0.0015],[494,0.0013],[496,0.0007],[498,0.0008],[500,0.0004],[502,0.001],[506,0.0013],[510,0.0015],[512,0.0013],[514,0.0004],[516,0],[518,0.0002],[522,0.002],[526,0.0031],[530,0.0038],[542,0.0036],[546,0.0026],[548,0.0017],[550,0.0012],[560,0.0011],[570,0.0014],[572,0.0008],[576,0.0008],[578,0.0012],[580,0.0012],[584,0.0009],[592,0.0018],[598,0.0021],[600,0.0026],[610,0.0027],[616,0.0032],[620,0.003],[622,0.0027],[642,0.0026],[646,0.0021],[648,0.0021],[652,0.0027],[658,0.0023],[666,0.0028],[680,0.0028],[684,0.0024],[690,0.0026],[696,0.0024],[704,0.0026],[706,0.0029],[712,0.0028],[714,0.0027],[716,0.0028],[720,0.0028]]
       }
       CTRL_L_eye_squintInner {
        type=MovieSceneScriptingFloatChannel
        default=0
        thinned_from=492
        keys [[-262,0.0015],[-260,0.001],[-258,0.0001],[-234,0],[-228,0.0004],[-222,0.0002],[-220,0],[-216,0.0002],[-214,0.0005],[-212,0.0005],[-206,0],[-196,0],[-190,0.0002],[-180,0],[-176,0.0004],[-168,0.002],[-164,0.0022],[-162,0.002],[-158,0.0021],[-156,0.0023],[-154,0.0023],[-150,0.0024],[-148,0.0022],[-144,0.0028],[-142,0.0026],[-134,0.0027],[-132,0.0026],[-130,0.0022],[-120,0.0023],[-118,0.0027],[-116,0.0028],[-112,0.0024],[-110,0.002],[-108,0.0019],[-96,0.0023],[-90,0.0027],[-84,0.0022],[-76,0.0026],[-72,0.0024],[-68,0.003],[-66,0.0076],[-64,0.0112],[-62,0.0136],[-60,0.0317],[-58,0.0277],[-56,0.02],[-54,0.0307],[-52,0.0353],[-50,0.023],[-48,0.0185],[-46,0.0069],[-44,0.0018],[-42,0.0018],[-40,0.0036],[-38,0.0078],[-36,0.0088],[-34,0.0049],[-32,0.0024],[-28,0.0023],[-26,0.0027],[-22,0.0028],[-20,0.0031],[-16,0.003],[-14,0.0032],[-8,0.0029],[-4,0.0031],[-2,0.0029],[2,0.0036],[4,0.0044],[6,0.0058],[8,0.0122],[10,0.0281],[12,0.0981],[14,0.1239],[16,0.1273],[18,0.1346],[20,0.1386],[22,0.1341],[24,0.1263],[26,0.1259],[28,0.1386],[32,0.1582],[34,0.1664],[36,0.1703],[38,0.1623],[40,0.1688],[42,0.1803],[44,0.1743],[46,0.1731],[48,0.1699],[50,0.1701],[52,0.1758],[54,0.1801],[58,0.1854],[60,0.1869],[62,0.1806],[64,0.169],[66,0.1719],[68,0.1741],[70,0.1743],[72,0.1767],[74,0.1802],[76,0.18],[80,0.1848],[82,0.1758],[84,0.163],[86,0.1547],[88,0.1496],[90,0.1488],[92,0.1366],[94,0.1083],[96,0.0791],[98,0.0747],[100,0.083],[102,0.0781],[104,0.0747],[106,0.0711],[108,0.0869],[110,0.1033],[112,0.0872],[114,0.0884],[116,0.1078],[118,0.099],[120,0.0952],[122,0.0878],[124,0.0865],[126,0.0871],[128,0.0962],[130,0.0962],[132,0.0863],[134,0.078],[136,0.0774],[138,0.0791],[140,0.0855],[142,0.0933],[144,0.1088],[146,0.1196],[148,0.1327],[150,0.1347],[152,0.145],[154,0.1561],[156,0.1565],[158,0.1629],[160,0.1739],[162,0.174],[164,0.1646],[166,0.1707],[168,0.1731],[170,0.1637],[172,0.157],[174,0.1653],[176,0.1694],[178,0.1686],[180,0.1619],[182,0.1546],[184,0.144],[186,0.146],[188,0.1508],[190,0.1502],[192,0.1455],[194,0.1464],[196,0.1494],[198,0.1479],[200,0.1459],[202,0.1379],[204,0.1343],[206,0.1256],[208,0.1299],[210,0.1302],[212,0.1177],[214,0.1113],[216,0.123],[218,0.1201],[220,0.1132],[222,0.1103],[224,0.1088],[226,0.1087],[228,0.1167],[230,0.1258],[232,0.1146],[234,0.0993],[236,0.0877],[238,0.0746],[240,0.0731],[242,0.1052],[244,0.1366],[246,0.1511],[248,0.1539],[250,0.1623],[252,0.1776],[254,0.1712],[256,0.1692],[258,0.1737],[260,0.1717],[264,0.1629],[266,0.1578],[268,0.1553],[270,0.1279],[272,0.0689],[274,0.0366],[276,0.0817],[278,0.1549],[280,0.1683],[282,0.1573],[284,0.1435],[286,0.1438],[288,0.1336],[290,0.1217],[292,0.1249],[294,0.1297],[298,0.1278],[300,0.1359],[302,0.1468],[304,0.1462],[306,0.1447],[308,0.1472],[310,0.1393],[312,0.1311],[314,0.138],[316,0.1286],[318,0.1219],[320,0.1255],[322,0.1328],[324,0.1242],[326,0.1107],[328,0.109],[330,0.0946],[332,0.0657],[334,0.0632],[336,0.0576],[338,0.0559],[340,0.0645],[342,0.0598],[344,0.0426],[346,0.0325],[350,0.011],[352,0.0071],[354,0.0067],[356,0.0056],[358,0.0029],[360,0.0006],[362,0.0008],[364,0.0006],[366,0.0001],[368,0],[372,0.0004],[376,0.0004],[380,0],[382,0],[384,0.0002],[388,0.0027],[390,0.0029],[392,0.0027],[394,0.0018],[396,0.0019],[398,0.0029],[400,0.0032],[402,0.0025],[404,0.009],[406,0.0101],[408,0.0031],[410,0.0025],[420,0.0022],[424,0.0023],[426,0.0019],[428,0.0017],[432,0.0018],[436,0.0015],[438,0.0016],[440,0.0021],[442,0.0023],[448,0.0018],[452,0.0022],[456,0.0021],[462,0.0016],[464,0.0016],[466,0.0021],[468,0.0022],[470,0.001],[474,0.0021],[476,0.0018],[478,0.0007],[480,0.0006],[482,0.0009],[484,0.001],[486,0.0002],[488,0],[492,0.0001],[506,0],[510,0.0001],[518,0],[522,0.0013],[524,0.0012],[526,0.0027],[530,0.0032],[532,0.0031],[538,0.0019],[542,0.0019],[544,0.0017],[548,0],[570,0.0001],[588,0],[592,0.0007],[594,0.0007],[596,0.0003],[598,0.0005],[600,0.0011],[608,0.001],[614,0.0016],[632,0.0014],[636,0.0007],[640,0.0013],[642,0.0015],[646,0.0007],[650,0.0017],[656,0.0016],[660,0.0016],[664,0.0021],[668,0.0017],[672,0.0018],[674,0.0017],[678,0.0018],[684,0.0016],[686,0.0017],[694,0.0016],[696,0.0017],[700,0.0011],[704,0.0017],[720,0.0017]]
       }
       CTRL_L_nose_wrinkleUpper {type=MovieSceneScriptingFloatChannel, default=0, const [1,492]}
       CTRL_R_nose_wrinkleUpper {type=MovieSceneScriptingFloatChannel, default=0, const [1,492]}
       CTRL_R_eye_squintInner {
        type=MovieSceneScriptingFloatChannel
        default=0
        thinned_from=492
        keys [[-262,0.0073],[-260,0.0052],[-258,0.0041],[-256,0.0043],[-252,0.0028],[-250,0.0022],[-248,0.0027],[-246,0.0019],[-244,0.0017],[-242,0.0004],[-240,0.003],[-238,0.003],[-234,0.0034],[-232,0.0052],[-230,0.006],[-228,0.0061],[-226,0.0054],[-224,0.0051],[-222,0.0055],[-220,0.0046],[-218,0.004],[-214,0.0054],[-212,0.0055],[-208,0.0054],[-206,0.005],[-202,0.003],[-200,0.0022],[-198,0.0029],[-194,0.0037],[-186,0.0031],[-184,0.0028],[-182,0.0028],[-180,0.0034],[-176,0.0041],[-174,0.0051],[-172,0.0069],[-168,0.0096],[-166,0.0106],[-164,0.0114],[-162,0.0129],[-160,0.0154],[-158,0.0156],[-154,0.0187],[-152,0.0194],[-150,0.0198],[-148,0.0206],[-146,0.022],[-144,0.0224],[-142,0.0221],[-140,0.0238],[-136,0.0264],[-134,0.0266],[-132,0.0257],[-130,0.0255],[-128,0.0264],[-126,0.0281],[-124,0.0294],[-122,0.0298],[-120,0.0299],[-118,0.031],[-116,0.0305],[-114,0.0322],[-112,0.0355],[-110,0.0355],[-108,0.0333],[-106,0.0313],[-104,0.0306],[-102,0.0328],[-100,0.0288],[-98,0.0283],[-96,0.0293],[-94,0.03],[-92,0.0293],[-90,0.0253],[-86,0.0259],[-84,0.0241],[-82,0.0258],[-80,0.0294],[-76,0.0283],[-74,0.0299],[-72,0.0321],[-68,0.0461],[-66,0.05],[-64,0.0551],[-62,0.0586],[-60,0.0762],[-58,0.0704],[-56,0.0593],[-54,0.0666],[-52,0.0653],[-50,0.0484],[-48,0.0425],[-46,0.025],[-44,0.0164],[-42,0.0161],[-40,0.022],[-38,0.0339],[-36,0.036],[-34,0.0288],[-32,0.0239],[-30,0.0236],[-28,0.0214],[-26,0.0221],[-24,0.0278],[-22,0.0293],[-20,0.0318],[-18,0.0358],[-16,0.0363],[-14,0.0392],[-12,0.0406],[-10,0.0423],[-8,0.041],[-6,0.0412],[-4,0.044],[-2,0.0449],[0,0.0466],[2,0.0495],[4,0.0562],[6,0.0578],[8,0.0679],[10,0.0892],[12,0.1519],[14,0.1698],[16,0.1724],[18,0.1783],[20,0.1815],[22,0.1755],[24,0.167],[26,0.17],[30,0.1994],[32,0.2085],[34,0.2149],[36,0.227],[38,0.2256],[40,0.2319],[42,0.2399],[44,0.2348],[46,0.2336],[48,0.2302],[50,0.2282],[52,0.2306],[54,0.2347],[56,0.2356],[58,0.2325],[60,0.2328],[62,0.2259],[64,0.2101],[66,0.2133],[68,0.2151],[70,0.215],[72,0.216],[74,0.2243],[78,0.2328],[80,0.2328],[82,0.219],[84,0.2066],[86,0.199],[88,0.1941],[90,0.194],[92,0.1804],[94,0.1433],[96,0.1015],[98,0.0931],[100,0.0982],[102,0.0869],[104,0.087],[106,0.0925],[108,0.1072],[110,0.1213],[112,0.1022],[114,0.1021],[116,0.1225],[118,0.1127],[120,0.1086],[122,0.1],[124,0.0931],[130,0.0931],[132,0.0911],[134,0.0833],[136,0.0739],[138,0.0728],[140,0.078],[142,0.083],[144,0.0908],[146,0.0972],[148,0.1154],[150,0.1189],[152,0.1199],[154,0.1296],[156,0.1302],[158,0.1396],[160,0.1475],[162,0.1519],[164,0.1454],[166,0.1407],[168,0.1393],[170,0.1336],[172,0.1265],[174,0.1298],[176,0.1318],[178,0.1261],[180,0.122],[182,0.1164],[184,0.1057],[186,0.1064],[188,0.1079],[190,0.1066],[192,0.1076],[194,0.1139],[196,0.1157],[198,0.109],[200,0.1095],[202,0.1078],[204,0.1056],[206,0.0891],[208,0.0951],[210,0.0956],[212,0.0935],[214,0.0878],[216,0.0927],[218,0.0892],[220,0.0853],[222,0.0865],[224,0.0814],[226,0.0799],[228,0.0868],[230,0.0902],[232,0.0777],[234,0.0649],[236,0.0571],[238,0.0481],[240,0.0459],[242,0.0672],[244,0.0898],[246,0.1026],[248,0.1082],[250,0.1141],[252,0.1258],[254,0.1236],[256,0.1229],[258,0.1267],[260,0.1237],[262,0.1201],[264,0.1169],[266,0.1112],[268,0.1103],[270,0.0934],[272,0.0551],[274,0.0428],[276,0.0816],[278,0.1305],[280,0.1417],[284,0.1252],[286,0.1274],[288,0.117],[290,0.1041],[292,0.1041],[294,0.1082],[296,0.1107],[298,0.1111],[300,0.1144],[302,0.1218],[304,0.1227],[306,0.121],[308,0.1273],[310,0.1263],[312,0.1171],[314,0.122],[316,0.1162],[318,0.1086],[320,0.1108],[322,0.1216],[324,0.1177],[326,0.1085],[328,0.105],[330,0.0904],[332,0.0646],[334,0.0614],[336,0.059],[338,0.0621],[340,0.0696],[342,0.0649],[344,0.0514],[346,0.0413],[348,0.0347],[350,0.0288],[352,0.0239],[354,0.0233],[356,0.0215],[360,0.0186],[362,0.0198],[364,0.0196],[366,0.0176],[368,0.0191],[370,0.0197],[372,0.0184],[374,0.0172],[376,0.0171],[378,0.0158],[380,0.0141],[382,0.0137],[384,0.0112],[386,0.011],[388,0.0103],[390,0.0086],[394,0.0042],[396,0.0026],[400,0.003],[402,0.0019],[406,0.0108],[408,0.0115],[410,0.0113],[412,0.0105],[416,0.0118],[418,0.0123],[420,0.014],[422,0.0149],[424,0.0144],[426,0.0137],[428,0.0127],[430,0.0128],[432,0.0138],[434,0.0136],[436,0.0138],[440,0.0148],[444,0.0148],[446,0.0151],[450,0.0142],[452,0.0143],[454,0.0149],[456,0.0147],[458,0.0137],[460,0.0133],[462,0.0135],[466,0.0132],[468,0.0137],[470,0.0136],[472,0.0193],[474,0.0165],[476,0.0131],[478,0.0101],[480,0.0125],[482,0.0175],[484,0.0183],[486,0.0214],[488,0.0226],[490,0.0166],[492,0.0094],[494,0.0073],[496,0.0073],[498,0.0069],[500,0.0059],[502,0.0073],[504,0.0083],[506,0.0076],[508,0.0093],[510,0.01],[514,0.0058],[516,0.0054],[518,0.0053],[520,0.007],[522,0.0122],[524,0.0146],[526,0.0302],[528,0.0329],[530,0.0315],[532,0.0244],[536,0.0127],[538,0.0103],[540,0.0101],[542,0.0128],[544,0.0083],[546,0.0048],[548,0.0031],[550,0.0029],[554,0.0039],[558,0.0043],[560,0.005],[568,0.005],[570,0.0052],[572,0.0032],[574,0.0031],[578,0.0038],[582,0.0029],[584,0.0057],[586,0.011],[590,0.0143],[592,0.0121],[594,0.0116],[596,0.0086],[598,0.0083],[600,0.0096],[602,0.0096],[604,0.0093],[606,0.0092],[610,0.008],[612,0.0083],[614,0.0095],[616,0.0103],[618,0.0109],[620,0.0111],[622,0.0107],[628,0.0105],[630,0.0101],[634,0.0101],[636,0.0098],[640,0.0106],[642,0.0109],[644,0.0102],[646,0.009],[648,0.0095],[650,0.0104],[652,0.0105],[654,0.0101],[656,0.0105],[658,0.0112],[660,0.0108],[662,0.0109],[664,0.0114],[666,0.011],[668,0.0111],[670,0.0113],[672,0.0112],[674,0.0109],[678,0.0112],[682,0.0109],[684,0.0109],[686,0.0116],[692,0.0111],[694,0.0103],[696,0.0105],[698,0.0111],[700,0.011],[702,0.0115],[704,0.0124],[706,0.013],[710,0.0134],[712,0.0134],[716,0.0127],[720,0.0127]]
       }
       CTRL_R_eye_cheekRaise {
        type=MovieSceneScriptingFloatChannel
        default=0
        thinned_from=492
        keys [[-262,0.0063],[-258,0.0052],[-256,0.0044],[-252,0.0034],[-250,0.0032],[-248,0.0034],[-244,0.0032],[-242,0.0028],[-240,0.0034],[-238,0.0036],[-236,0.0033],[-234,0.0033],[-230,0.0042],[-228,0.0044],[-224,0.0041],[-220,0.004],[-218,0.0037],[-216,0.0038],[-214,0.0043],[-210,0.0046],[-208,0.0046],[-206,0.0042],[-204,0.0042],[-198,0.0035],[-192,0.0041],[-188,0.0044],[-186,0.0044],[-182,0.0038],[-172,0.0059],[-170,0.0062],[-168,0.0061],[-160,0.0044],[-154,0.0041],[-152,0.0043],[-148,0.0042],[-134,0.0036],[-130,0.0036],[-126,0.0032],[-122,0.0033],[-120,0.0036],[-118,0.0046],[-116,0.0064],[-114,0.007],[-112,0.0083],[-110,0.0093],[-108,0.0093],[-106,0.0088],[-104,0.0086],[-102,0.0089],[-98,0.0086],[-96,0.008],[-94,0.0084],[-88,0.0079],[-82,0.0058],[-80,0.0055],[-78,0.0049],[-76,0.0047],[-74,0.004],[-72,0.0038],[-68,0.0051],[-66,0.0065],[-64,0.0082],[-62,0.0089],[-60,0.0123],[-58,0.0122],[-52,0.0113],[-50,0.0097],[-48,0.0091],[-46,0.007],[-44,0.0059],[-42,0.0057],[-40,0.0062],[-36,0.0051],[-34,0.0044],[-32,0.0045],[-30,0.005],[-28,0.005],[-26,0.0053],[-20,0.0076],[-18,0.0083],[-16,0.0093],[-14,0.0111],[-12,0.0125],[-10,0.0142],[-8,0.0139],[-6,0.0144],[-4,0.0142],[-2,0.0152],[0,0.0158],[2,0.0171],[4,0.0193],[6,0.019],[8,0.019],[10,0.0185],[12,0.0193],[16,0.0183],[18,0.0175],[20,0.0177],[24,0.0171],[26,0.0172],[28,0.0177],[30,0.0188],[32,0.0196],[36,0.0206],[38,0.0227],[40,0.0239],[44,0.0245],[46,0.0242],[48,0.0235],[50,0.0232],[52,0.0224],[54,0.0226],[56,0.0225],[58,0.0213],[60,0.0206],[62,0.0193],[64,0.0185],[66,0.0167],[68,0.0162],[70,0.0161],[72,0.0157],[74,0.0162],[78,0.0195],[80,0.02],[82,0.0183],[84,0.0169],[86,0.0158],[88,0.015],[90,0.0147],[92,0.0129],[94,0.0107],[96,0.0069],[98,0.0054],[100,0.0046],[102,0.0044],[104,0.0049],[106,0.0058],[108,0.0055],[112,0.004],[116,0.0021],[118,0.0016],[120,0.0015],[124,0.0023],[136,0.0024],[138,0.0022],[140,0.0023],[142,0.002],[146,0.001],[156,0.0008],[158,0.0005],[166,0.0005],[170,0],[198,0],[202,0.0004],[204,0.0004],[206,0],[350,0],[354,0.0003],[356,0.0006],[362,0.0002],[364,0.0003],[368,0.0001],[374,0.0012],[384,0.0013],[386,0.0024],[388,0.0048],[390,0.0049],[392,0.0043],[394,0.0031],[396,0.0024],[400,0.0024],[402,0.0028],[406,0.0031],[408,0.0037],[412,0.0041],[414,0.0041],[422,0.005],[430,0.0047],[438,0.005],[440,0.0055],[452,0.0054],[456,0.0059],[464,0.0054],[470,0.0043],[474,0.0063],[476,0.0064],[478,0.0057],[480,0.0056],[482,0.0061],[486,0.0061],[488,0.0055],[494,0.0047],[496,0.004],[498,0.0041],[500,0.0038],[502,0.0042],[508,0.0038],[512,0.0041],[514,0.0028],[516,0.002],[518,0.0022],[522,0.0048],[524,0.0059],[526,0.0067],[528,0.0071],[530,0.0079],[534,0.0078],[536,0.0075],[540,0.0076],[544,0.0064],[546,0.0055],[548,0.0044],[550,0.0039],[552,0.0038],[560,0.0037],[566,0.0042],[570,0.0041],[572,0.0033],[576,0.0035],[578,0.004],[580,0.004],[582,0.0037],[584,0.0039],[588,0.0059],[590,0.0065],[594,0.0067],[596,0.0062],[598,0.006],[600,0.0063],[602,0.0062],[604,0.0059],[606,0.0058],[610,0.0053],[620,0.0055],[622,0.0052],[634,0.0052],[636,0.005],[638,0.005],[642,0.0055],[646,0.005],[648,0.005],[652,0.0056],[654,0.0054],[670,0.0053],[672,0.0055],[680,0.0051],[682,0.0049],[704,0.005],[706,0.0052],[714,0.0051],[716,0.0052],[718,0.0049],[720,0.0049]]
       }
       CTRL_C_eyelidsReduce {type=MovieSceneScriptingFloatChannel, default=0, const [0,492]}
       CTRL_R_brow_lateral {
        type=MovieSceneScriptingFloatChannel
        default=0
        thinned_from=492
        keys [[-262,0.006],[-260,0.0061],[-258,0.0063],[-252,0.0059],[-244,0.0059],[-242,0.0053],[-240,0.0055],[-238,0.0059],[-236,0.0056],[-230,0.006],[-228,0.0064],[-222,0.0064],[-218,0.0067],[-216,0.006],[-214,0.0056],[-208,0.0053],[-204,0.0046],[-198,0.0049],[-184,0.0048],[-176,0.0055],[-170,0.0055],[-162,0.0061],[-160,0.0067],[-158,0.0067],[-156,0.0071],[-154,0.0072],[-150,0.0064],[-148,0.0054],[-144,0.0047],[-138,0.0042],[-136,0.0044],[-126,0.004],[-120,0.0034],[-116,0.0035],[-114,0.0033],[-110,0.0024],[-104,0.0021],[-102,0.0023],[-100,0.0027],[-92,0.003],[-90,0.0033],[-88,0.0033],[-82,0.0043],[-78,0.0058],[-76,0.0059],[-74,0.0059],[-70,0.0047],[-68,0.0045],[-66,0.0045],[-62,0.0039],[-60,0.0028],[-56,0.0025],[-48,0.0029],[-46,0.0029],[-40,0.0034],[-38,0.0046],[-36,0.0153],[-34,0.0315],[-32,0.0536],[-30,0.0657],[-26,0.08],[-24,0.0952],[-22,0.1109],[-20,0.1262],[-18,0.1444],[-16,0.1582],[-14,0.1672],[-12,0.1751],[-10,0.1785],[-8,0.1831],[-6,0.185],[-4,0.192],[-2,0.1945],[0,0.2002],[2,0.2071],[4,0.2169],[6,0.2159],[8,0.2111],[10,0.2137],[12,0.2207],[14,0.2173],[16,0.2154],[18,0.2109],[22,0.2082],[26,0.2065],[28,0.2064],[30,0.2053],[32,0.203],[34,0.1994],[36,0.1993],[38,0.1976],[40,0.1967],[42,0.195],[44,0.1939],[46,0.1934],[48,0.1915],[50,0.1876],[52,0.1825],[56,0.1746],[58,0.1703],[60,0.1664],[62,0.1629],[64,0.1598],[66,0.1598],[68,0.1611],[70,0.161],[72,0.1598],[74,0.1557],[76,0.1567],[78,0.1597],[80,0.1641],[82,0.1634],[84,0.1597],[86,0.1566],[88,0.1547],[90,0.1542],[92,0.1405],[94,0.1073],[96,0.0577],[98,0.0318],[100,0.0186],[102,0.0088],[106,0.0011],[120,0.0007],[124,0.0011],[126,0.0011],[128,0.0015],[132,0.0015],[134,0.0014],[136,0.0015],[142,0.0008],[144,0.0007],[150,0.0013],[154,0.0009],[158,0.0009],[160,0.0011],[162,0.0016],[168,0.0019],[174,0.002],[178,0.0023],[182,0.0023],[184,0.002],[198,0.0018],[204,0.0016],[206,0.002],[218,0.002],[220,0.0023],[222,0.0024],[224,0.002],[226,0.002],[228,0.0017],[230,0.0019],[232,0.0018],[234,0.002],[236,0.002],[238,0.0017],[242,0.0016],[244,0.0019],[248,0.002],[250,0.0017],[254,0.0017],[256,0.0022],[260,0.0022],[264,0.0017],[268,0.0017],[270,0.0022],[274,0.0027],[278,0.0019],[280,0.0012],[286,0.0008],[290,0.0009],[292,0.0013],[302,0.0015],[308,0.0009],[310,0.001],[314,0.0016],[316,0.0014],[318,0.0009],[324,0.0011],[326,0.0011],[332,0.0013],[334,0.0011],[338,0.0013],[346,0.0012],[348,0.001],[350,0.0011],[352,0.0015],[356,0.0016],[360,0.0011],[362,0.0006],[364,0.0007],[368,0.0013],[372,0.0016],[376,0.0015],[380,0.0019],[382,0.0024],[386,0.0024],[388,0.0028],[390,0.0028],[392,0.0024],[394,0.0022],[398,0.0035],[404,0.0037],[406,0.0033],[408,0.0033],[410,0.0028],[430,0.0025],[438,0.0026],[446,0.0029],[452,0.0025],[456,0.0028],[466,0.0025],[468,0.0027],[470,0.0031],[474,0.0052],[478,0.0067],[480,0.0069],[482,0.0063],[484,0.0062],[488,0.0055],[492,0.0061],[496,0.0075],[498,0.0076],[502,0.0069],[504,0.0069],[506,0.0067],[510,0.0056],[512,0.0059],[520,0.0056],[522,0.007],[524,0.01],[526,0.0185],[530,0.0194],[532,0.0176],[534,0.0145],[538,0.0074],[540,0.0069],[542,0.005],[544,0.0048],[552,0.0059],[556,0.0059],[560,0.0063],[562,0.0062],[564,0.0059],[568,0.0059],[570,0.006],[572,0.0056],[578,0.0059],[580,0.0057],[582,0.005],[584,0.0048],[588,0.0061],[590,0.006],[592,0.0052],[594,0.0051],[598,0.0042],[600,0.0032],[602,0.0031],[604,0.0028],[610,0.0023],[616,0.0027],[622,0.0022],[626,0.0022],[628,0.0024],[636,0.0022],[644,0.0024],[646,0.0024],[648,0.0025],[652,0.0024],[654,0.0022],[674,0.0027],[680,0.0024],[684,0.0021],[690,0.0024],[694,0.0022],[698,0.0024],[704,0.0028],[716,0.0025],[720,0.0024]]
       }
       CTRL_L_brow_lateral {
        type=MovieSceneScriptingFloatChannel
        default=0
        thinned_from=492
        keys [[-262,0.0075],[-258,0.0084],[-256,0.0084],[-254,0.0079],[-252,0.0078],[-248,0.0079],[-246,0.0076],[-244,0.0076],[-242,0.007],[-240,0.0071],[-238,0.0074],[-234,0.0073],[-230,0.0077],[-228,0.0081],[-218,0.0081],[-214,0.0073],[-208,0.0072],[-204,0.0067],[-194,0.0074],[-190,0.0072],[-188,0.0069],[-180,0.0068],[-176,0.0072],[-164,0.0071],[-154,0.0068],[-152,0.0066],[-148,0.0053],[-144,0.005],[-138,0.0047],[-130,0.0047],[-128,0.0045],[-120,0.0047],[-116,0.0049],[-106,0.0039],[-102,0.0038],[-98,0.004],[-94,0.0045],[-82,0.0051],[-78,0.0059],[-74,0.006],[-70,0.0051],[-66,0.005],[-62,0.005],[-54,0.0043],[-48,0.0043],[-44,0.0049],[-42,0.0056],[-40,0.0056],[-38,0.0059],[-36,0.0147],[-34,0.0295],[-32,0.0509],[-30,0.0631],[-28,0.0701],[-26,0.0766],[-24,0.0902],[-20,0.1186],[-18,0.1357],[-16,0.1489],[-14,0.1573],[-12,0.1646],[-10,0.1671],[-8,0.1716],[-6,0.1727],[-4,0.1794],[-2,0.1818],[0,0.1876],[2,0.1942],[4,0.2031],[6,0.202],[8,0.1973],[10,0.2005],[12,0.2081],[14,0.2052],[16,0.2036],[18,0.1998],[22,0.1973],[28,0.1953],[30,0.193],[32,0.1904],[34,0.1873],[36,0.1882],[38,0.1858],[40,0.1853],[42,0.1835],[46,0.1826],[48,0.1805],[50,0.1771],[52,0.1721],[56,0.1638],[58,0.16],[62,0.1537],[64,0.1509],[66,0.1513],[68,0.153],[70,0.153],[72,0.1519],[74,0.1476],[76,0.1481],[78,0.1504],[80,0.1546],[82,0.1546],[86,0.1484],[88,0.1467],[90,0.1464],[92,0.1337],[94,0.1026],[96,0.056],[98,0.0318],[100,0.0196],[102,0.0107],[106,0.0041],[112,0.0039],[116,0.0035],[122,0.0036],[126,0.0039],[128,0.0045],[132,0.0044],[136,0.0041],[142,0.0037],[144,0.0038],[146,0.0044],[150,0.0047],[156,0.0046],[160,0.005],[162,0.0055],[164,0.0055],[168,0.0059],[172,0.0056],[176,0.0059],[178,0.0063],[182,0.0067],[184,0.0063],[190,0.0065],[198,0.0058],[200,0.0055],[204,0.0055],[206,0.006],[210,0.006],[216,0.0057],[220,0.0063],[222,0.0063],[224,0.006],[228,0.0059],[230,0.0061],[234,0.0059],[238,0.0053],[240,0.0051],[246,0.0059],[248,0.006],[250,0.0059],[254,0.0061],[256,0.0064],[260,0.0063],[262,0.0061],[264,0.0062],[268,0.0059],[272,0.006],[274,0.0058],[276,0.0053],[284,0.0046],[290,0.0046],[294,0.0052],[304,0.0053],[308,0.0047],[310,0.0046],[314,0.0053],[316,0.0052],[318,0.0047],[324,0.0049],[326,0.0047],[330,0.0049],[334,0.0046],[338,0.0052],[340,0.0049],[346,0.005],[350,0.0046],[352,0.0049],[356,0.005],[360,0.0044],[362,0.004],[372,0.0047],[378,0.0044],[380,0.0044],[382,0.0049],[384,0.0047],[390,0.0044],[392,0.0041],[394,0.0043],[398,0.0057],[400,0.0059],[408,0.0058],[410,0.0053],[414,0.0049],[426,0.0046],[430,0.0044],[438,0.0046],[440,0.005],[452,0.0047],[456,0.005],[458,0.005],[460,0.0047],[464,0.0046],[468,0.005],[476,0.0069],[478,0.0078],[480,0.0083],[482,0.0077],[484,0.0076],[488,0.0066],[492,0.0071],[496,0.0088],[498,0.0092],[502,0.0086],[504,0.0086],[510,0.0073],[512,0.0076],[514,0.0073],[520,0.0071],[522,0.0081],[524,0.0102],[526,0.0156],[530,0.0158],[532,0.0148],[534,0.0129],[536,0.0106],[538,0.0085],[540,0.0082],[542,0.0072],[544,0.0071],[546,0.0072],[550,0.008],[552,0.0082],[556,0.0081],[560,0.0084],[562,0.0083],[566,0.0077],[570,0.0078],[572,0.0077],[578,0.008],[580,0.0078],[582,0.0073],[584,0.0072],[588,0.0082],[590,0.0082],[592,0.0077],[594,0.0076],[598,0.0064],[600,0.0053],[602,0.0053],[604,0.0049],[606,0.0048],[608,0.0044],[616,0.0047],[634,0.0043],[636,0.0042],[642,0.0046],[652,0.005],[658,0.0048],[664,0.0052],[668,0.0049],[680,0.005],[684,0.0047],[690,0.0049],[698,0.0044],[714,0.0047],[720,0.0046]]
       }
       CTRL_R_brow_down {
        type=MovieSceneScriptingFloatChannel
        default=0
        thinned_from=492
        keys [[-262,0.0067],[-258,0.0062],[-252,0.0045],[-250,0.0046],[-248,0.0051],[-244,0.0052],[-242,0.0045],[-240,0.0045],[-238,0.005],[-234,0.0049],[-228,0.006],[-220,0.0065],[-216,0.0063],[-214,0.0068],[-212,0.0069],[-210,0.0072],[-208,0.0073],[-204,0.0067],[-200,0.0065],[-198,0.0063],[-194,0.0067],[-190,0.0067],[-188,0.0064],[-174,0.0058],[-170,0.0064],[-168,0.0065],[-162,0.0061],[-160,0.0066],[-158,0.0066],[-156,0.0071],[-150,0.0118],[-148,0.0126],[-146,0.0125],[-142,0.014],[-140,0.0145],[-138,0.0143],[-134,0.0145],[-130,0.0144],[-126,0.0151],[-120,0.0148],[-118,0.0149],[-114,0.0161],[-112,0.0175],[-110,0.0192],[-108,0.02],[-104,0.0195],[-102,0.0183],[-100,0.018],[-98,0.018],[-96,0.0175],[-94,0.0166],[-88,0.0165],[-86,0.0158],[-84,0.0135],[-82,0.0116],[-80,0.01],[-78,0.0095],[-76,0.0094],[-74,0.0114],[-72,0.0129],[-70,0.0148],[-66,0.0167],[-64,0.0179],[-62,0.0183],[-60,0.021],[-56,0.0209],[-54,0.0218],[-52,0.0233],[-50,0.0236],[-48,0.0234],[-46,0.0182],[-44,0.0124],[-42,0.0075],[-40,0.0079],[-38,0.0105],[-36,0.0188],[-34,0.0283],[-32,0.0404],[-30,0.0467],[-26,0.0556],[-22,0.0776],[-20,0.0896],[-18,0.1027],[-16,0.1133],[-14,0.1206],[-12,0.1275],[-8,0.1342],[-6,0.1359],[-4,0.1412],[-2,0.1427],[0,0.1479],[2,0.1537],[4,0.1624],[6,0.1616],[8,0.1587],[10,0.1617],[12,0.1735],[14,0.1747],[16,0.1741],[18,0.1726],[20,0.172],[24,0.1695],[26,0.1689],[30,0.1708],[32,0.1695],[34,0.1671],[36,0.1686],[38,0.1707],[42,0.1705],[44,0.1692],[46,0.1694],[48,0.1688],[50,0.1648],[52,0.1611],[54,0.158],[60,0.1515],[64,0.1439],[66,0.1431],[68,0.1448],[70,0.1447],[72,0.1434],[74,0.1414],[76,0.1447],[78,0.1499],[80,0.153],[82,0.1492],[84,0.1446],[88,0.1409],[90,0.1401],[92,0.1243],[94,0.0974],[96,0.0572],[98,0.038],[100,0.0263],[102,0.0177],[106,0.0115],[108,0.0112],[116,0.0064],[118,0.0056],[122,0.0053],[124,0.0061],[126,0.0061],[128,0.0068],[132,0.007],[138,0.0068],[142,0.0072],[144,0.0077],[146,0.0083],[152,0.0097],[158,0.0096],[160,0.01],[162,0.0107],[166,0.0115],[168,0.0111],[170,0.011],[172,0.0103],[174,0.0104],[176,0.0103],[178,0.0105],[184,0.0103],[190,0.0108],[192,0.0108],[194,0.0106],[198,0.0108],[200,0.0114],[202,0.0116],[204,0.0115],[208,0.0108],[210,0.0107],[214,0.0098],[216,0.0098],[218,0.0099],[222,0.0086],[224,0.0074],[226,0.0066],[228,0.0056],[234,0.0044],[238,0.0031],[242,0.003],[244,0.0033],[246,0.0029],[248,0.0026],[250,0.0016],[254,0.0013],[256,0.0018],[258,0.0019],[262,0.0013],[264,0.0008],[266,0.0008],[270,0.0016],[276,0.0005],[278,0.0004],[280,0.001],[282,0.0008],[284,0.0013],[288,0.0014],[290,0.0007],[292,0.0003],[298,0.0004],[300,0.0002],[306,0.001],[308,0.001],[310,0.0013],[312,0.0013],[316,0.0009],[326,0.0019],[334,0.0019],[336,0.0017],[338,0.0022],[342,0.0026],[346,0.0022],[348,0.0016],[350,0.0014],[354,0.0018],[356,0.0022],[360,0.0017],[362,0.0013],[364,0.0014],[368,0.0013],[370,0.0014],[374,0.002],[382,0.0024],[384,0.0029],[386,0.004],[388,0.0049],[390,0.0046],[394,0.002],[396,0.0011],[398,0.0005],[402,0.0004],[404,0.0006],[406,0.0011],[408,0.0018],[410,0.002],[412,0.0025],[414,0.0027],[420,0.0041],[422,0.0045],[428,0.0042],[438,0.0048],[446,0.0058],[450,0.0059],[452,0.0058],[454,0.006],[466,0.005],[468,0.005],[470,0.0045],[474,0.005],[476,0.0049],[478,0.0043],[480,0.0043],[482,0.0047],[484,0.0048],[486,0.0055],[490,0.0058],[494,0.0048],[496,0.0041],[498,0.0039],[500,0.0033],[502,0.0037],[506,0.0029],[508,0.0023],[510,0.002],[514,0.0029],[518,0.0033],[520,0.0048],[522,0.0074],[524,0.014],[526,0.0355],[528,0.039],[530,0.0455],[532,0.042],[536,0.0211],[538,0.0129],[540,0.0118],[542,0.0071],[544,0.0055],[548,0.004],[556,0.0035],[560,0.0043],[562,0.0043],[564,0.0038],[568,0.0037],[570,0.0039],[572,0.0029],[574,0.0029],[576,0.0032],[578,0.0041],[580,0.0042],[582,0.0039],[584,0.0041],[586,0.0051],[588,0.0056],[590,0.0053],[592,0.0048],[596,0.0046],[598,0.0048],[600,0.0057],[602,0.0058],[604,0.0072],[612,0.0078],[614,0.0082],[616,0.0084],[620,0.0082],[626,0.0087],[628,0.0094],[630,0.0094],[632,0.0093],[634,0.0089],[638,0.0087],[640,0.0091],[642,0.0093],[646,0.009],[652,0.0097],[654,0.0093],[658,0.0098],[660,0.0104],[664,0.0107],[670,0.0108],[674,0.0111],[682,0.01],[684,0.01],[686,0.0103],[690,0.0105],[700,0.01],[704,0.0106],[706,0.0112],[710,0.0119],[716,0.0123],[718,0.012],[720,0.012]]
       }
       CTRL_L_brow_down {
        type=MovieSceneScriptingFloatChannel
        default=0
        thinned_from=492
        keys [[-262,0.0087],[-258,0.0079],[-256,0.0072],[-252,0.0064],[-248,0.0068],[-244,0.0068],[-242,0.006],[-240,0.0064],[-238,0.007],[-236,0.0068],[-232,0.0071],[-228,0.0078],[-220,0.0081],[-218,0.0079],[-216,0.008],[-214,0.0083],[-208,0.0086],[-198,0.0077],[-190,0.008],[-188,0.0077],[-186,0.0076],[-182,0.0069],[-180,0.0071],[-170,0.0091],[-164,0.0095],[-162,0.0091],[-160,0.0094],[-156,0.0094],[-150,0.0122],[-148,0.0126],[-146,0.0123],[-140,0.0137],[-134,0.0139],[-128,0.014],[-126,0.0142],[-118,0.0142],[-116,0.0151],[-114,0.0156],[-110,0.0187],[-108,0.0194],[-104,0.019],[-102,0.0184],[-98,0.0182],[-94,0.017],[-88,0.0169],[-86,0.016],[-84,0.014],[-82,0.0124],[-80,0.011],[-78,0.0104],[-76,0.0103],[-74,0.012],[-72,0.0129],[-70,0.0142],[-68,0.0147],[-66,0.0157],[-64,0.0168],[-62,0.0174],[-60,0.0209],[-58,0.0209],[-56,0.0212],[-54,0.0219],[-52,0.0232],[-50,0.0234],[-48,0.0232],[-46,0.0184],[-44,0.0127],[-42,0.0081],[-40,0.0086],[-38,0.011],[-36,0.0192],[-34,0.0283],[-32,0.0408],[-30,0.0476],[-28,0.0521],[-26,0.0561],[-24,0.0656],[-22,0.0754],[-20,0.0862],[-18,0.0982],[-16,0.108],[-14,0.1148],[-12,0.1209],[-10,0.1234],[-8,0.1266],[-6,0.1275],[-4,0.1324],[-2,0.134],[0,0.1395],[2,0.1453],[4,0.1532],[6,0.1522],[8,0.1492],[10,0.1528],[12,0.1643],[14,0.1656],[16,0.1653],[18,0.1645],[20,0.1643],[24,0.1621],[26,0.1615],[30,0.1621],[32,0.1605],[34,0.1583],[36,0.1605],[38,0.1618],[40,0.1626],[42,0.1626],[46,0.1621],[48,0.161],[52,0.1537],[54,0.1505],[56,0.1483],[60,0.1452],[64,0.1386],[66,0.1381],[68,0.1401],[70,0.14],[72,0.1388],[74,0.1367],[76,0.1394],[78,0.1436],[80,0.1464],[82,0.1436],[84,0.1394],[86,0.1372],[88,0.1356],[90,0.1351],[92,0.1206],[94,0.0954],[96,0.0578],[98,0.0397],[100,0.0288],[102,0.0201],[104,0.017],[106,0.0142],[108,0.0139],[110,0.0129],[116,0.0092],[118,0.0085],[120,0.0084],[122,0.008],[124,0.0089],[126,0.0089],[128,0.0094],[130,0.0094],[132,0.0096],[140,0.0094],[144,0.0096],[150,0.0114],[152,0.0118],[160,0.0118],[162,0.0126],[164,0.0127],[166,0.0132],[168,0.0129],[170,0.0129],[172,0.0122],[174,0.0124],[176,0.0123],[178,0.0125],[180,0.012],[192,0.0128],[196,0.0125],[198,0.0127],[200,0.0133],[202,0.0134],[204,0.0134],[208,0.0128],[210,0.0128],[212,0.0119],[214,0.0114],[216,0.0113],[218,0.0116],[222,0.0106],[224,0.0093],[226,0.0087],[228,0.0074],[232,0.0068],[238,0.0061],[240,0.0064],[242,0.0063],[244,0.0063],[246,0.0056],[248,0.0056],[250,0.0048],[254,0.0044],[258,0.0047],[260,0.0046],[264,0.0038],[266,0.0037],[268,0.0044],[270,0.0048],[272,0.0045],[274,0.0037],[278,0.0029],[280,0.0032],[282,0.0029],[284,0.0031],[286,0.0031],[288,0.0037],[290,0.0035],[292,0.003],[294,0.0032],[296,0.003],[298,0.0031],[300,0.0029],[302,0.0033],[308,0.0033],[310,0.0037],[314,0.0037],[316,0.0034],[318,0.0037],[320,0.0036],[326,0.0041],[332,0.0038],[334,0.004],[336,0.0037],[338,0.0044],[342,0.0053],[346,0.0051],[348,0.0043],[350,0.004],[354,0.0045],[356,0.0051],[362,0.0036],[364,0.004],[370,0.004],[374,0.0046],[382,0.0045],[384,0.0051],[388,0.0072],[390,0.0074],[392,0.007],[396,0.0054],[398,0.005],[404,0.0046],[406,0.0051],[408,0.006],[410,0.0062],[422,0.0082],[426,0.0081],[434,0.0075],[438,0.0077],[440,0.0083],[444,0.0085],[452,0.0086],[456,0.0089],[462,0.0085],[468,0.0085],[472,0.0067],[478,0.0055],[480,0.0057],[482,0.0066],[484,0.0068],[486,0.0077],[490,0.0078],[492,0.0072],[494,0.007],[496,0.0064],[498,0.0067],[500,0.0059],[502,0.0059],[504,0.0052],[508,0.0042],[510,0.0039],[512,0.0044],[518,0.0047],[520,0.0062],[522,0.0084],[524,0.0133],[526,0.029],[528,0.0319],[530,0.0369],[532,0.0341],[534,0.0255],[536,0.0174],[538,0.0117],[540,0.011],[542,0.0082],[544,0.0074],[546,0.0069],[548,0.0061],[554,0.0059],[556,0.0056],[562,0.0061],[564,0.0058],[570,0.0059],[572,0.0051],[574,0.0051],[576,0.0053],[578,0.0061],[580,0.0061],[582,0.0057],[584,0.0057],[586,0.0061],[588,0.0063],[590,0.006],[594,0.006],[596,0.0063],[598,0.0069],[600,0.008],[602,0.0081],[604,0.0089],[610,0.009],[616,0.0097],[622,0.0092],[624,0.0092],[628,0.0097],[630,0.0098],[636,0.0091],[638,0.009],[640,0.0095],[642,0.0097],[644,0.0095],[646,0.0091],[652,0.0102],[654,0.0098],[664,0.0106],[670,0.0107],[672,0.011],[674,0.011],[682,0.0099],[690,0.0106],[696,0.01],[700,0.0099],[704,0.0107],[706,0.0115],[710,0.012],[716,0.0121],[718,0.0119],[720,0.0119]]
       }
       CTRL_L_brow_raiseOut {
        type=MovieSceneScriptingFloatChannel
        default=0.7687
        thinned_from=492
        keys [[-262,0.5975],[-260,0.6191],[-258,0.619],[-256,0.6114],[-252,0.607],[-250,0.609],[-248,0.6049],[-246,0.5976],[-244,0.596],[-242,0.5958],[-240,0.6014],[-238,0.5842],[-236,0.5763],[-234,0.5737],[-232,0.5566],[-230,0.5527],[-228,0.557],[-226,0.5674],[-224,0.5682],[-222,0.5558],[-220,0.5376],[-218,0.5272],[-216,0.5224],[-214,0.519],[-212,0.5183],[-210,0.5126],[-206,0.5131],[-202,0.5397],[-200,0.5447],[-198,0.5323],[-196,0.5219],[-192,0.5128],[-190,0.5114],[-188,0.517],[-186,0.5162],[-184,0.5089],[-182,0.4934],[-178,0.4803],[-176,0.4817],[-174,0.4847],[-172,0.4814],[-170,0.4695],[-168,0.4502],[-166,0.4413],[-164,0.4402],[-162,0.4453],[-160,0.4372],[-158,0.4365],[-156,0.4295],[-154,0.4171],[-150,0.4061],[-148,0.4053],[-146,0.4145],[-142,0.4106],[-140,0.4026],[-138,0.401],[-134,0.4111],[-132,0.4098],[-130,0.4005],[-128,0.3948],[-126,0.3942],[-122,0.3975],[-120,0.3954],[-118,0.3892],[-116,0.3776],[-114,0.3777],[-112,0.3762],[-110,0.3728],[-108,0.3744],[-106,0.3807],[-102,0.3887],[-100,0.3861],[-98,0.3857],[-96,0.3925],[-94,0.4141],[-92,0.4133],[-90,0.4083],[-88,0.4062],[-86,0.4029],[-84,0.4146],[-82,0.42],[-80,0.4186],[-76,0.4205],[-74,0.4182],[-72,0.4096],[-70,0.3966],[-68,0.3934],[-66,0.3886],[-64,0.3933],[-62,0.3932],[-60,0.3869],[-58,0.3865],[-54,0.3798],[-52,0.3732],[-50,0.3814],[-48,0.3876],[-46,0.4836],[-44,0.6237],[-42,0.7883],[-40,0.8616],[-38,0.8561],[-36,0.7814],[-34,0.7152],[-32,0.678],[-30,0.6603],[-28,0.6446],[-26,0.6369],[-24,0.6182],[-20,0.5691],[-16,0.5367],[-14,0.518],[-12,0.498],[-10,0.4956],[-8,0.4948],[-6,0.4982],[-4,0.4934],[-2,0.4952],[0,0.4766],[2,0.4631],[4,0.4465],[6,0.4502],[8,0.4565],[10,0.4498],[12,0.4326],[14,0.4323],[16,0.433],[18,0.4352],[20,0.435],[22,0.4285],[24,0.4194],[26,0.4162],[28,0.4274],[30,0.4371],[32,0.4352],[34,0.4296],[36,0.4179],[38,0.4125],[40,0.4026],[42,0.3962],[44,0.4008],[46,0.4026],[50,0.4149],[52,0.4182],[54,0.425],[56,0.4287],[58,0.4225],[60,0.4137],[62,0.4173],[64,0.4309],[66,0.4367],[68,0.4356],[70,0.4358],[74,0.4408],[76,0.4313],[78,0.4208],[80,0.4179],[82,0.4234],[86,0.4253],[88,0.4338],[90,0.436],[92,0.5066],[94,0.5504],[96,0.6344],[98,0.6652],[100,0.7269],[102,0.7534],[104,0.7653],[106,0.7365],[108,0.7247],[110,0.6982],[112,0.6736],[114,0.6504],[116,0.642],[118,0.6462],[120,0.6476],[122,0.6436],[124,0.6152],[126,0.6124],[128,0.5679],[130,0.5679],[132,0.5487],[134,0.543],[136,0.5395],[138,0.5321],[140,0.5173],[144,0.4914],[146,0.4612],[148,0.4441],[152,0.4134],[154,0.4053],[158,0.4047],[160,0.408],[162,0.4056],[164,0.4001],[166,0.3921],[168,0.3928],[172,0.4052],[176,0.3991],[178,0.3931],[180,0.3926],[182,0.3852],[184,0.3822],[186,0.3807],[188,0.3759],[190,0.3685],[192,0.3652],[194,0.3696],[196,0.3749],[198,0.3701],[200,0.3617],[204,0.3656],[206,0.3572],[208,0.3595],[210,0.3598],[212,0.3782],[214,0.3905],[216,0.3895],[230,0.3906],[232,0.3915],[234,0.3934],[236,0.3973],[238,0.4029],[244,0.4244],[246,0.4334],[248,0.4436],[252,0.4701],[254,0.4865],[256,0.501],[258,0.5192],[260,0.5396],[262,0.5688],[264,0.6024],[266,0.642],[268,0.7084],[270,0.7883],[272,0.8804],[274,0.9476],[276,0.9989],[278,0.9815],[280,0.9449],[282,0.9274],[284,0.9038],[286,0.9154],[288,0.9365],[290,0.9459],[292,0.9584],[294,0.9533],[296,0.9504],[298,0.9349],[300,0.9236],[302,0.9061],[304,0.895],[310,0.8925],[312,0.8847],[314,0.8801],[318,0.862],[322,0.8568],[324,0.8518],[326,0.8495],[328,0.8446],[330,0.8468],[332,0.8397],[336,0.8419],[338,0.841],[342,0.8347],[344,0.8335],[346,0.8282],[348,0.8269],[350,0.8322],[352,0.8405],[354,0.8384],[356,0.8284],[358,0.8126],[360,0.801],[362,0.7969],[364,0.7968],[366,0.7924],[368,0.7775],[370,0.7539],[372,0.7345],[374,0.727],[378,0.7239],[380,0.7119],[382,0.6954],[384,0.6538],[386,0.6594],[388,0.8082],[390,0.8652],[392,0.9584],[394,0.9931],[396,1],[420,1],[422,0.9957],[424,0.994],[426,0.9851],[428,0.9822],[430,0.9733],[432,0.9619],[434,0.9435],[440,0.8956],[442,0.8836],[444,0.8762],[446,0.8654],[448,0.8564],[450,0.8493],[452,0.8478],[456,0.8357],[460,0.8202],[462,0.8199],[464,0.8175],[466,0.8336],[468,0.8356],[470,0.8542],[472,0.8768],[474,0.8944],[476,0.8911],[478,0.8507],[480,0.7908],[482,0.7284],[484,0.717],[486,0.7014],[488,0.6609],[490,0.5932],[492,0.5034],[494,0.4759],[496,0.4606],[498,0.4613],[500,0.4776],[502,0.5013],[504,0.5232],[508,0.5306],[510,0.5517],[512,0.5529],[514,0.5434],[516,0.5555],[518,0.5509],[520,0.4944],[522,0.429],[524,0.3585],[526,0.2912],[528,0.2797],[530,0.2533],[532,0.259],[534,0.2961],[536,0.3851],[538,0.4875],[540,0.5026],[542,0.6119],[544,0.6664],[546,0.7078],[548,0.753],[550,0.7621],[552,0.7533],[554,0.7373],[556,0.7235],[560,0.686],[562,0.6772],[570,0.6736],[572,0.6865],[574,0.6834],[576,0.6765],[578,0.6591],[580,0.655],[582,0.6548],[584,0.6523],[586,0.6514],[588,0.6664],[590,0.6778],[592,0.6794],[594,0.6787],[596,0.6578],[600,0.6478],[602,0.6333],[604,0.4843],[606,0.4629],[608,0.4048],[610,0.4008],[612,0.4061],[614,0.4048],[618,0.3903],[620,0.3881],[622,0.3731],[624,0.3714],[626,0.369],[628,0.3659],[632,0.3723],[634,0.3763],[636,0.3752],[638,0.375],[640,0.3712],[642,0.3695],[644,0.3689],[646,0.3745],[648,0.3772],[650,0.3767],[652,0.3744],[654,0.3751],[656,0.3703],[658,0.3684],[660,0.3617],[662,0.3604],[664,0.353],[668,0.3472],[670,0.3462],[672,0.3436],[674,0.3403],[676,0.3421],[678,0.3447],[680,0.353],[682,0.3555],[684,0.3542],[688,0.3439],[690,0.3438],[694,0.354],[696,0.3576],[698,0.3571],[700,0.3584],[702,0.3568],[704,0.3505],[706,0.3382],[708,0.3338],[710,0.3307],[712,0.3298],[716,0.3254],[720,0.325]]
       }
       CTRL_L_brow_raiseIn {
        type=MovieSceneScriptingFloatChannel
        default=0.2801
        thinned_from=492
        keys [[-262,0.3081],[-260,0.3159],[-258,0.3163],[-256,0.3019],[-254,0.2991],[-252,0.2957],[-250,0.2967],[-248,0.2901],[-246,0.2795],[-244,0.2771],[-242,0.2779],[-240,0.2838],[-238,0.2714],[-236,0.2687],[-234,0.2672],[-232,0.2582],[-230,0.2596],[-228,0.2598],[-226,0.2654],[-224,0.2648],[-222,0.2579],[-220,0.2446],[-218,0.236],[-216,0.2296],[-214,0.2258],[-212,0.2254],[-210,0.2239],[-206,0.2181],[-204,0.22],[-202,0.2327],[-200,0.2387],[-198,0.2342],[-196,0.226],[-194,0.2261],[-192,0.2247],[-190,0.2238],[-188,0.2214],[-186,0.2211],[-184,0.2236],[-182,0.2133],[-180,0.2085],[-178,0.2063],[-176,0.2157],[-174,0.2274],[-172,0.2316],[-170,0.2281],[-168,0.2198],[-166,0.2167],[-164,0.2217],[-162,0.2255],[-160,0.2208],[-158,0.22],[-156,0.2125],[-150,0.1852],[-148,0.1849],[-146,0.1949],[-144,0.1946],[-142,0.1893],[-140,0.1789],[-138,0.1802],[-136,0.1781],[-134,0.1847],[-132,0.1802],[-130,0.1767],[-128,0.1711],[-126,0.1678],[-124,0.1672],[-122,0.1703],[-120,0.173],[-118,0.1708],[-116,0.1606],[-114,0.1598],[-112,0.1572],[-110,0.1533],[-108,0.1531],[-106,0.1589],[-104,0.1627],[-102,0.1765],[-100,0.1751],[-98,0.1747],[-96,0.1733],[-94,0.1898],[-92,0.1894],[-90,0.1859],[-86,0.1833],[-84,0.1956],[-82,0.2015],[-80,0.1993],[-78,0.1988],[-74,0.2002],[-72,0.1893],[-70,0.172],[-68,0.1698],[-66,0.1621],[-64,0.1632],[-62,0.1626],[-60,0.1594],[-58,0.1593],[-56,0.1584],[-52,0.1664],[-50,0.1881],[-48,0.1979],[-46,0.2675],[-44,0.3867],[-42,0.5371],[-40,0.6544],[-38,0.7203],[-36,0.7431],[-34,0.7589],[-32,0.7834],[-30,0.8008],[-26,0.8018],[-24,0.7972],[-22,0.7896],[-20,0.7798],[-18,0.7902],[-16,0.7895],[-14,0.7793],[-12,0.7578],[-10,0.7501],[-6,0.7527],[-2,0.7515],[0,0.7356],[2,0.7232],[4,0.7078],[6,0.705],[8,0.7058],[10,0.6988],[12,0.673],[14,0.6614],[16,0.6581],[20,0.6457],[22,0.6319],[24,0.6209],[26,0.6196],[28,0.6317],[30,0.6338],[32,0.623],[34,0.6101],[36,0.5942],[40,0.5653],[42,0.5518],[44,0.5576],[46,0.5599],[48,0.57],[50,0.582],[54,0.5754],[56,0.5649],[58,0.543],[60,0.5256],[62,0.5236],[64,0.537],[66,0.5438],[68,0.5447],[70,0.5442],[72,0.5394],[74,0.5281],[78,0.5021],[80,0.5034],[82,0.5098],[84,0.51],[86,0.5054],[88,0.5138],[90,0.5172],[92,0.5934],[94,0.6118],[96,0.6544],[98,0.656],[100,0.7054],[102,0.711],[104,0.7058],[106,0.6544],[108,0.6325],[112,0.5257],[114,0.4755],[116,0.4426],[118,0.4198],[120,0.4165],[122,0.3992],[124,0.3575],[126,0.3542],[128,0.3014],[130,0.3014],[132,0.2851],[136,0.2788],[138,0.2689],[140,0.253],[142,0.243],[144,0.2364],[146,0.2181],[148,0.2052],[150,0.1902],[152,0.1789],[154,0.172],[156,0.1721],[158,0.1747],[160,0.1769],[162,0.1755],[164,0.1715],[166,0.1644],[168,0.1671],[170,0.1731],[172,0.185],[174,0.1853],[176,0.1851],[178,0.1777],[180,0.1742],[182,0.1662],[186,0.162],[188,0.1558],[190,0.1491],[192,0.1456],[194,0.1497],[196,0.1519],[198,0.1461],[200,0.134],[202,0.1318],[204,0.1323],[206,0.1267],[208,0.1289],[210,0.1292],[212,0.1483],[214,0.1574],[226,0.1573],[230,0.1579],[232,0.1586],[234,0.1606],[236,0.1648],[238,0.171],[240,0.1787],[244,0.192],[246,0.2002],[248,0.2097],[250,0.2229],[252,0.2356],[254,0.251],[256,0.2631],[258,0.2793],[260,0.297],[262,0.3221],[264,0.3548],[266,0.3953],[268,0.4692],[270,0.5584],[272,0.6625],[274,0.7511],[276,0.7864],[278,0.7491],[280,0.6809],[282,0.662],[284,0.6384],[286,0.6459],[288,0.6575],[292,0.6693],[294,0.6632],[296,0.6535],[298,0.6352],[300,0.6202],[302,0.6014],[306,0.5831],[308,0.5771],[310,0.5675],[314,0.5541],[318,0.5332],[322,0.5401],[324,0.5419],[326,0.5362],[328,0.5335],[330,0.5345],[332,0.5332],[334,0.5337],[336,0.5363],[338,0.5359],[342,0.5225],[344,0.5222],[348,0.5123],[350,0.5138],[352,0.5227],[354,0.5223],[356,0.5183],[358,0.5083],[360,0.5005],[362,0.5055],[364,0.5156],[366,0.5157],[368,0.4975],[370,0.4716],[372,0.4531],[374,0.4492],[378,0.4485],[380,0.4426],[382,0.4336],[384,0.414],[386,0.4318],[388,0.5817],[390,0.6499],[392,0.8049],[394,0.9141],[396,0.9823],[398,1],[402,0.9994],[404,0.9615],[406,0.939],[408,0.9136],[410,0.9234],[412,0.9076],[414,0.8985],[416,0.8826],[418,0.8559],[420,0.8204],[422,0.7817],[424,0.7729],[426,0.7557],[428,0.7518],[432,0.7237],[434,0.7053],[436,0.6887],[438,0.6697],[440,0.655],[444,0.6383],[446,0.6242],[448,0.6181],[450,0.6177],[452,0.6219],[456,0.6151],[458,0.6152],[460,0.6169],[462,0.6252],[464,0.6284],[466,0.6411],[468,0.6361],[470,0.6261],[472,0.605],[474,0.5912],[476,0.5794],[478,0.545],[480,0.4921],[482,0.4304],[484,0.4187],[486,0.4054],[488,0.3829],[490,0.3412],[492,0.2793],[494,0.2541],[496,0.2376],[498,0.2291],[500,0.2465],[502,0.2633],[504,0.2894],[506,0.2883],[508,0.2994],[510,0.3184],[512,0.3195],[514,0.2996],[516,0.2902],[518,0.2839],[520,0.2145],[522,0.1549],[524,0.103],[526,0.0702],[528,0.0652],[530,0.0523],[532,0.052],[534,0.0684],[536,0.1164],[538,0.18],[540,0.1899],[542,0.28],[544,0.3751],[546,0.4374],[548,0.4984],[550,0.5076],[552,0.497],[556,0.4609],[560,0.4162],[562,0.4056],[564,0.401],[566,0.3947],[568,0.3939],[570,0.3952],[572,0.4164],[574,0.413],[576,0.4054],[578,0.3882],[580,0.3845],[582,0.3818],[584,0.364],[586,0.3326],[588,0.3236],[590,0.3339],[592,0.346],[594,0.3479],[596,0.3512],[598,0.3537],[600,0.3472],[602,0.3336],[604,0.1946],[606,0.176],[608,0.1258],[610,0.1224],[612,0.1271],[614,0.1254],[618,0.1143],[620,0.1122],[622,0.0963],[624,0.094],[626,0.0893],[628,0.0852],[630,0.0856],[632,0.0879],[634,0.0909],[638,0.0881],[640,0.0853],[642,0.0841],[644,0.0835],[646,0.0855],[648,0.087],[650,0.0865],[652,0.0852],[654,0.0852],[656,0.0828],[658,0.081],[660,0.0774],[662,0.0763],[664,0.0728],[666,0.0709],[668,0.0695],[670,0.069],[672,0.068],[674,0.0665],[676,0.0672],[678,0.0677],[680,0.0707],[682,0.0717],[684,0.0718],[686,0.0693],[688,0.0671],[690,0.0669],[694,0.0697],[696,0.0702],[698,0.0704],[700,0.0713],[702,0.0714],[704,0.0694],[706,0.0647],[708,0.0625],[710,0.0608],[712,0.0604],[714,0.0595],[716,0.0592],[718,0.0584],[720,0.0584]]
       }
       CTRL_R_brow_raiseIn {
        type=MovieSceneScriptingFloatChannel
        default=0.43
        thinned_from=492
        keys [[-262,0.2543],[-260,0.2569],[-258,0.2555],[-256,0.2425],[-254,0.2429],[-252,0.2411],[-250,0.2387],[-248,0.2306],[-246,0.2217],[-244,0.2198],[-242,0.2215],[-240,0.2279],[-238,0.2192],[-236,0.2193],[-234,0.2185],[-232,0.214],[-230,0.2169],[-228,0.2153],[-226,0.2199],[-224,0.2211],[-222,0.2169],[-220,0.2048],[-218,0.1945],[-216,0.1851],[-214,0.1788],[-210,0.1771],[-208,0.1733],[-206,0.1713],[-204,0.1705],[-202,0.1844],[-200,0.1888],[-198,0.1847],[-196,0.1764],[-194,0.178],[-190,0.176],[-188,0.1687],[-186,0.1688],[-184,0.177],[-182,0.1707],[-180,0.1699],[-178,0.1686],[-176,0.1791],[-174,0.1913],[-172,0.1961],[-170,0.1918],[-168,0.1829],[-166,0.1799],[-162,0.1896],[-160,0.1864],[-158,0.1861],[-156,0.1828],[-154,0.1738],[-152,0.1629],[-150,0.1485],[-148,0.1482],[-146,0.1584],[-144,0.1586],[-142,0.1523],[-140,0.1419],[-138,0.1445],[-136,0.1381],[-134,0.1448],[-132,0.1403],[-130,0.1397],[-128,0.1347],[-126,0.1306],[-124,0.1295],[-120,0.1374],[-118,0.1354],[-116,0.123],[-114,0.1217],[-112,0.1182],[-110,0.1133],[-108,0.1109],[-106,0.115],[-104,0.1181],[-102,0.1355],[-100,0.1367],[-98,0.1366],[-96,0.1339],[-94,0.1473],[-92,0.1471],[-90,0.1456],[-88,0.1446],[-86,0.1444],[-84,0.1563],[-82,0.1626],[-80,0.1616],[-78,0.1609],[-76,0.1612],[-74,0.1623],[-72,0.1515],[-70,0.1345],[-68,0.1318],[-66,0.1257],[-64,0.1257],[-62,0.1248],[-60,0.1203],[-58,0.1196],[-56,0.1182],[-54,0.1231],[-52,0.1299],[-50,0.1517],[-48,0.1609],[-46,0.2177],[-44,0.3161],[-42,0.4506],[-40,0.5736],[-38,0.6362],[-36,0.6564],[-34,0.6593],[-32,0.6821],[-30,0.7023],[-26,0.7007],[-24,0.6882],[-22,0.6779],[-20,0.6641],[-18,0.6701],[-16,0.6676],[-14,0.6576],[-12,0.6351],[-10,0.6255],[-8,0.626],[-4,0.6192],[-2,0.619],[2,0.5997],[4,0.5848],[6,0.5797],[8,0.5786],[10,0.5757],[12,0.5555],[14,0.5448],[16,0.5423],[20,0.5342],[22,0.5216],[24,0.5124],[26,0.5127],[28,0.5225],[30,0.5209],[34,0.4994],[36,0.4899],[38,0.4772],[40,0.4689],[42,0.4572],[44,0.4655],[46,0.4666],[48,0.4731],[50,0.485],[52,0.4803],[54,0.4741],[56,0.4617],[58,0.4432],[60,0.4322],[62,0.4314],[64,0.4436],[66,0.4476],[68,0.449],[70,0.4486],[72,0.4444],[74,0.4332],[76,0.4198],[78,0.4052],[80,0.4067],[82,0.4152],[84,0.4175],[86,0.4118],[88,0.4187],[90,0.4224],[92,0.4926],[94,0.5155],[96,0.5704],[98,0.5839],[100,0.6417],[102,0.6589],[104,0.6617],[106,0.622],[108,0.6003],[114,0.4606],[116,0.4261],[118,0.4005],[120,0.3968],[122,0.383],[124,0.3482],[126,0.345],[128,0.2953],[130,0.2953],[132,0.2838],[134,0.2822],[136,0.2799],[138,0.2687],[140,0.2527],[142,0.2443],[144,0.2384],[146,0.2233],[148,0.2113],[150,0.1965],[152,0.1844],[154,0.1773],[156,0.1775],[158,0.1799],[160,0.1813],[162,0.1795],[164,0.1755],[166,0.1654],[168,0.1689],[170,0.1751],[172,0.191],[174,0.1894],[176,0.189],[178,0.1818],[180,0.1787],[182,0.1711],[184,0.1706],[186,0.1686],[188,0.1627],[190,0.1561],[192,0.1526],[194,0.1566],[196,0.158],[198,0.1511],[200,0.1367],[202,0.1321],[204,0.1319],[206,0.1286],[208,0.1317],[210,0.1319],[212,0.1514],[214,0.16],[228,0.1606],[230,0.161],[232,0.1617],[234,0.1636],[236,0.1675],[238,0.1733],[240,0.1804],[244,0.1924],[246,0.1998],[248,0.2085],[250,0.2203],[252,0.2316],[254,0.246],[256,0.2575],[258,0.2737],[260,0.2908],[262,0.3131],[264,0.3423],[266,0.3779],[268,0.4477],[270,0.5307],[272,0.6275],[274,0.7092],[276,0.7356],[278,0.6946],[280,0.6309],[282,0.6175],[284,0.6084],[288,0.6302],[290,0.6382],[292,0.6433],[294,0.6411],[296,0.6329],[298,0.6147],[300,0.5994],[302,0.5814],[304,0.5763],[306,0.5664],[308,0.5586],[310,0.5451],[312,0.5352],[314,0.5286],[316,0.5192],[318,0.5116],[322,0.5194],[324,0.5218],[326,0.5136],[328,0.511],[334,0.5148],[336,0.517],[338,0.5177],[342,0.5021],[344,0.4985],[346,0.4908],[348,0.485],[350,0.4863],[352,0.494],[354,0.4936],[356,0.4897],[358,0.4796],[360,0.4726],[362,0.4805],[364,0.4939],[366,0.4941],[368,0.4748],[370,0.4476],[372,0.4309],[374,0.4285],[376,0.4284],[378,0.43],[380,0.4227],[382,0.412],[384,0.3837],[386,0.3876],[388,0.5164],[390,0.5835],[392,0.7453],[394,0.8724],[396,0.9643],[398,0.997],[400,0.9999],[402,0.999],[404,0.9356],[406,0.891],[408,0.8494],[410,0.8715],[412,0.8604],[414,0.8529],[416,0.8383],[418,0.8121],[420,0.7786],[422,0.7413],[424,0.7336],[426,0.7178],[428,0.7157],[430,0.6997],[432,0.6865],[438,0.6348],[440,0.6228],[444,0.6078],[446,0.5933],[448,0.5897],[450,0.5897],[452,0.5943],[454,0.5918],[456,0.5908],[460,0.5956],[462,0.6054],[464,0.6089],[466,0.6208],[468,0.6174],[470,0.6114],[472,0.5888],[478,0.5031],[480,0.4704],[482,0.4237],[484,0.4142],[486,0.3935],[488,0.3763],[490,0.3361],[492,0.276],[494,0.2468],[496,0.2298],[498,0.2211],[500,0.239],[502,0.2552],[504,0.2794],[506,0.2777],[508,0.2907],[510,0.3106],[512,0.3128],[514,0.2882],[516,0.2709],[518,0.265],[520,0.1976],[522,0.1336],[524,0.0744],[526,0.0278],[528,0.0256],[530,0.0194],[532,0.0189],[534,0.0288],[536,0.0646],[538,0.1167],[540,0.1249],[542,0.2023],[544,0.3075],[546,0.3748],[548,0.4434],[550,0.4544],[552,0.4496],[554,0.4381],[556,0.4233],[560,0.3819],[562,0.3733],[564,0.3699],[566,0.3653],[568,0.3648],[570,0.3676],[572,0.3885],[574,0.3857],[576,0.3788],[578,0.3619],[582,0.3534],[584,0.3275],[586,0.28],[588,0.2614],[590,0.2758],[592,0.2937],[594,0.2966],[596,0.3049],[598,0.3086],[600,0.3046],[602,0.2925],[604,0.168],[606,0.1518],[608,0.1078],[610,0.1047],[612,0.1093],[614,0.1074],[616,0.102],[618,0.0956],[620,0.0931],[622,0.0739],[624,0.0709],[628,0.0608],[630,0.0615],[632,0.0632],[634,0.0668],[636,0.0658],[638,0.0658],[640,0.0653],[642,0.0653],[644,0.0646],[646,0.0652],[648,0.0672],[650,0.0686],[652,0.0693],[654,0.0715],[656,0.071],[658,0.0677],[660,0.0636],[662,0.0622],[666,0.0603],[670,0.0609],[672,0.0614],[674,0.0607],[676,0.0628],[678,0.0639],[680,0.0658],[684,0.0665],[686,0.0653],[688,0.0644],[690,0.0642],[694,0.0667],[698,0.0665],[700,0.0672],[702,0.0688],[704,0.0689],[706,0.0679],[708,0.0673],[710,0.0655],[712,0.0649],[714,0.0627],[716,0.0619],[720,0.0618]]
       }
       CTRL_R_brow_raiseOut {
        type=MovieSceneScriptingFloatChannel
        default=0
        thinned_from=492
        keys [[-262,0.4486],[-260,0.4601],[-256,0.45],[-254,0.4553],[-252,0.459],[-250,0.4551],[-248,0.4472],[-246,0.4419],[-244,0.441],[-242,0.4441],[-240,0.4507],[-238,0.4409],[-236,0.439],[-234,0.438],[-232,0.4316],[-230,0.4308],[-228,0.4311],[-226,0.4401],[-224,0.445],[-222,0.4388],[-218,0.4074],[-216,0.395],[-214,0.3854],[-212,0.384],[-210,0.3786],[-208,0.3762],[-206,0.3799],[-204,0.3889],[-202,0.406],[-200,0.4072],[-198,0.3962],[-196,0.3864],[-194,0.3859],[-192,0.3822],[-190,0.3806],[-188,0.3758],[-186,0.376],[-184,0.3841],[-182,0.3788],[-180,0.3811],[-178,0.3762],[-176,0.3799],[-174,0.3828],[-172,0.3789],[-170,0.363],[-168,0.3399],[-166,0.3293],[-164,0.327],[-162,0.3334],[-160,0.3296],[-158,0.3299],[-156,0.3324],[-154,0.3236],[-152,0.316],[-150,0.302],[-148,0.3041],[-146,0.315],[-144,0.3166],[-142,0.3138],[-140,0.308],[-138,0.3101],[-136,0.3059],[-134,0.3094],[-132,0.3074],[-128,0.2996],[-126,0.2976],[-124,0.298],[-122,0.3006],[-120,0.3005],[-118,0.2964],[-116,0.2859],[-114,0.2858],[-112,0.2847],[-110,0.2799],[-108,0.2774],[-106,0.2813],[-104,0.2841],[-102,0.2968],[-100,0.2985],[-98,0.2985],[-96,0.3],[-94,0.3143],[-92,0.3138],[-88,0.3103],[-86,0.3067],[-84,0.314],[-82,0.318],[-78,0.3129],[-76,0.3125],[-74,0.3105],[-72,0.3028],[-70,0.2908],[-66,0.2849],[-64,0.2892],[-62,0.2896],[-60,0.2906],[-58,0.2901],[-56,0.2888],[-54,0.2895],[-52,0.291],[-50,0.3021],[-48,0.3077],[-46,0.3784],[-44,0.4746],[-42,0.6086],[-40,0.6994],[-38,0.6965],[-36,0.6213],[-34,0.5299],[-32,0.4879],[-30,0.4729],[-28,0.4524],[-26,0.4406],[-24,0.4062],[-20,0.3425],[-18,0.3167],[-14,0.2768],[-12,0.2549],[-10,0.247],[-8,0.2434],[-6,0.2378],[-4,0.2285],[-2,0.2298],[0,0.222],[2,0.2151],[4,0.2008],[6,0.2009],[8,0.2049],[10,0.2059],[12,0.1992],[16,0.2028],[18,0.2089],[20,0.2124],[22,0.2085],[24,0.203],[26,0.2025],[28,0.2091],[30,0.2118],[32,0.2099],[36,0.2092],[38,0.2053],[40,0.2061],[42,0.2029],[44,0.2116],[46,0.2112],[48,0.2116],[50,0.2182],[52,0.2194],[54,0.2219],[58,0.2212],[60,0.2235],[62,0.2284],[64,0.2393],[66,0.2397],[70,0.2398],[74,0.2438],[76,0.2343],[78,0.221],[80,0.2186],[82,0.2282],[84,0.2335],[86,0.2319],[88,0.2363],[90,0.2388],[92,0.2985],[94,0.3533],[96,0.4661],[98,0.5231],[100,0.6047],[102,0.6563],[104,0.6837],[106,0.6793],[108,0.6686],[110,0.6548],[112,0.6446],[114,0.6296],[116,0.6183],[118,0.6167],[120,0.6172],[122,0.6196],[124,0.6043],[126,0.6021],[128,0.5671],[130,0.5671],[132,0.5587],[134,0.5575],[136,0.5548],[138,0.5457],[140,0.5317],[142,0.5246],[144,0.5145],[146,0.4925],[150,0.4624],[152,0.4464],[154,0.4395],[156,0.4394],[158,0.438],[160,0.4403],[162,0.4364],[164,0.4308],[166,0.4166],[168,0.4191],[170,0.4253],[172,0.4399],[174,0.4337],[176,0.43],[178,0.4234],[180,0.423],[182,0.4164],[184,0.4163],[186,0.4153],[188,0.4111],[190,0.4038],[192,0.4007],[196,0.4074],[198,0.4016],[200,0.3892],[202,0.3859],[204,0.386],[206,0.3819],[208,0.3855],[210,0.3857],[212,0.4022],[214,0.412],[230,0.4131],[232,0.4139],[234,0.4156],[236,0.4191],[238,0.4239],[244,0.4423],[246,0.4497],[248,0.4583],[252,0.4795],[256,0.5069],[258,0.525],[260,0.5443],[262,0.568],[264,0.5955],[266,0.6272],[268,0.6893],[270,0.7717],[272,0.8711],[274,0.9481],[276,0.9727],[278,0.9221],[280,0.8633],[282,0.8474],[284,0.8408],[286,0.8584],[288,0.8794],[292,0.9039],[294,0.9059],[296,0.9061],[298,0.8898],[300,0.8782],[302,0.8624],[308,0.8512],[310,0.8422],[312,0.8295],[316,0.8166],[318,0.8123],[322,0.8094],[324,0.8043],[326,0.7967],[328,0.7914],[330,0.7941],[332,0.7896],[334,0.7905],[336,0.7889],[338,0.7895],[340,0.7844],[344,0.7698],[346,0.7586],[348,0.7558],[352,0.7655],[354,0.7634],[356,0.7529],[358,0.7359],[360,0.7258],[362,0.7267],[364,0.7327],[366,0.7279],[368,0.7106],[370,0.6848],[372,0.6687],[374,0.6634],[376,0.6628],[378,0.6643],[380,0.6495],[382,0.6302],[384,0.5776],[386,0.5628],[388,0.6849],[390,0.7484],[392,0.8757],[394,0.951],[396,0.9908],[398,1],[416,1],[418,0.9988],[420,0.984],[422,0.9579],[424,0.9499],[426,0.9345],[428,0.9358],[432,0.9126],[436,0.8774],[438,0.8634],[442,0.8445],[444,0.8364],[446,0.8238],[450,0.8142],[452,0.8137],[456,0.806],[460,0.7942],[462,0.7957],[464,0.7926],[466,0.8061],[468,0.8125],[470,0.8388],[472,0.8539],[474,0.8333],[476,0.7889],[478,0.7541],[480,0.7314],[482,0.694],[484,0.6857],[486,0.6482],[488,0.6157],[490,0.5501],[492,0.4657],[494,0.4304],[496,0.4131],[498,0.4119],[500,0.4289],[502,0.4507],[504,0.4703],[506,0.4744],[508,0.4852],[510,0.5071],[512,0.5099],[514,0.4951],[516,0.4972],[518,0.494],[520,0.4464],[522,0.3688],[524,0.2742],[526,0.1579],[528,0.1508],[530,0.1358],[532,0.1412],[534,0.1655],[536,0.2333],[538,0.3192],[540,0.3323],[542,0.4185],[544,0.4919],[546,0.5425],[548,0.6065],[550,0.6218],[552,0.6252],[554,0.6229],[556,0.6149],[558,0.5984],[560,0.583],[562,0.5774],[564,0.5774],[570,0.5804],[572,0.5939],[574,0.5923],[576,0.5866],[578,0.569],[580,0.5647],[582,0.5622],[584,0.5445],[586,0.5132],[588,0.5118],[590,0.5356],[592,0.5508],[594,0.5525],[596,0.5458],[598,0.548],[600,0.555],[602,0.5439],[604,0.4247],[606,0.4085],[608,0.3644],[610,0.3622],[612,0.3685],[614,0.3672],[616,0.3605],[618,0.3522],[620,0.3492],[622,0.324],[624,0.3206],[626,0.3164],[628,0.3096],[630,0.3135],[632,0.3163],[634,0.3231],[636,0.3231],[638,0.3259],[642,0.3273],[644,0.3265],[646,0.3289],[648,0.3339],[650,0.338],[652,0.3408],[654,0.3458],[656,0.3449],[658,0.338],[660,0.3294],[662,0.3267],[664,0.325],[666,0.3243],[668,0.326],[672,0.3274],[674,0.3258],[676,0.3316],[678,0.336],[680,0.3423],[682,0.3438],[684,0.3432],[686,0.3405],[688,0.3384],[690,0.3385],[694,0.3476],[696,0.3495],[698,0.3486],[700,0.3495],[702,0.3512],[704,0.3493],[706,0.3446],[708,0.3434],[710,0.3395],[712,0.3379],[714,0.333],[716,0.3296],[718,0.3309],[720,0.3309]]
       }
       CTRL_lookAtSwitch {type=MovieSceneScriptingFloatChannel, default=0, const [0,492]}
       CTRL_C_teethD.X {type=MovieSceneScriptingFloatChannel, default=0, const [0,492]}
       CTRL_C_teethD.Y {type=MovieSceneScriptingFloatChannel, default=0, const [0,492]}
       CTRL_C_teethU.X {type=MovieSceneScriptingFloatChannel, default=0, const [0,492]}
       CTRL_C_teethU.Y {type=MovieSceneScriptingFloatChannel, default=0, const [0,492]}
       CTRL_C_teeth_fwdBackU {type=MovieSceneScriptingFloatChannel, default=0, const [0,492]}
       CTRL_L_mouth_corner.X {type=MovieSceneScriptingFloatChannel, default=0, const [0,492]}
       CTRL_L_mouth_corner.Y {
        type=MovieSceneScriptingFloatChannel
        default=0
        thinned_from=492
        keys [[-262,0.0323],[-260,0.03],[-258,0.0295],[-256,0.03],[-254,0.03],[-252,0.0298],[-250,0.0305],[-246,0.0329],[-244,0.0332],[-240,0.0332],[-238,0.0335],[-234,0.0335],[-232,0.0328],[-230,0.0318],[-226,0.029],[-222,0.0266],[-220,0.0258],[-216,0.0251],[-214,0.0242],[-210,0.0237],[-208,0.0232],[-206,0.0232],[-204,0.0224],[-202,0.0218],[-200,0.0215],[-198,0.0223],[-196,0.0223],[-194,0.0212],[-190,0.0215],[-188,0.0212],[-186,0.0212],[-184,0.0217],[-182,0.0247],[-180,0.0282],[-176,0.0386],[-174,0.0432],[-170,0.049],[-168,0.0531],[-166,0.0581],[-164,0.064],[-162,0.0674],[-160,0.0693],[-158,0.0694],[-156,0.0696],[-154,0.0685],[-152,0.0681],[-150,0.07],[-148,0.0742],[-146,0.0791],[-144,0.0821],[-142,0.084],[-138,0.0866],[-134,0.0902],[-130,0.0931],[-128,0.093],[-126,0.0934],[-122,0.0982],[-120,0.0978],[-118,0.0882],[-116,0.0664],[-114,0.062],[-112,0.0521],[-110,0.0458],[-108,0.0417],[-106,0.0386],[-104,0.037],[-102,0.0337],[-100,0.0326],[-98,0.0326],[-96,0.0306],[-94,0.0309],[-92,0.0318],[-90,0.0367],[-88,0.0378],[-86,0.0415],[-82,0.052],[-78,0.0674],[-76,0.0696],[-74,0.0794],[-72,0.0831],[-70,0.0882],[-68,0.092],[-66,0.0935],[-64,0.087],[-62,0.0829],[-60,0.0575],[-58,0.0537],[-56,0.0456],[-54,0.0418],[-52,0.04],[-50,0.0388],[-48,0.0384],[-46,0.0372],[-44,0.0361],[-42,0.0359],[-40,0.0375],[-38,0.0384],[-36,0.04],[-34,0.0394],[-32,0.0383],[-30,0.0368],[-28,0.0362],[-26,0.0365],[-24,0.0371],[-22,0.0365],[-20,0.0365],[-18,0.0361],[-16,0.0361],[-10,0.0351],[-8,0.0342],[-2,0.0322],[2,0.0316],[4,0.0316],[6,0.0309],[8,0.0309],[10,0.0316],[12,0.0332],[14,0.0335],[20,0.0333],[24,0.0324],[26,0.0324],[30,0.0333],[32,0.0333],[34,0.0337],[36,0.0336],[38,0.0342],[40,0.0342],[42,0.0352],[50,0.0353],[54,0.0361],[58,0.0361],[62,0.0358],[64,0.0353],[66,0.0352],[68,0.0358],[70,0.0357],[72,0.0351],[74,0.0347],[76,0.0348],[78,0.0355],[80,0.0352],[84,0.0333],[86,0.0332],[88,0.0335],[90,0.0335],[92,0.0328],[94,0.0316],[96,0.0307],[98,0.0309],[102,0.0286],[104,0.0283],[106,0.0329],[110,0.0333],[112,0.0326],[116,0.0321],[118,0.0316],[122,0.0317],[124,0.0342],[126,0.0342],[128,0.0348],[132,0.0348],[134,0.0345],[136,0.0348],[138,0.0348],[140,0.0355],[142,0.0352],[144,0.0352],[146,0.0354],[148,0.0368],[150,0.0387],[152,0.0402],[154,0.0397],[156,0.0396],[158,0.0381],[160,0.0368],[162,0.0368],[164,0.0362],[166,0.0365],[168,0.0361],[170,0.0364],[174,0.0338],[176,0.0319],[178,0.0319],[180,0.0313],[182,0.0326],[184,0.0335],[186,0.0333],[188,0.0327],[190,0.0324],[192,0.0324],[196,0.0321],[198,0.0322],[200,0.0331],[202,0.0333],[204,0.0332],[206,0.0324],[210,0.0322],[212,0.032],[214,0.0322],[216,0.0319],[218,0.0326],[220,0.0335],[222,0.0352],[224,0.0355],[226,0.0355],[228,0.0349],[232,0.0361],[234,0.0379],[236,0.0403],[240,0.0432],[242,0.0417],[244,0.0413],[246,0.0414],[248,0.0423],[252,0.041],[254,0.0394],[256,0.0394],[260,0.0371],[266,0.0348],[268,0.0352],[270,0.0368],[272,0.0421],[274,0.0433],[276,0.0401],[278,0.0361],[280,0.0368],[282,0.0385],[284,0.0415],[292,0.0401],[294,0.041],[298,0.0416],[300,0.0407],[302,0.04],[306,0.0394],[308,0.0394],[312,0.0387],[314,0.0381],[316,0.0371],[318,0.0368],[324,0.0367],[326,0.0365],[328,0.0374],[330,0.0371],[332,0.0371],[334,0.0368],[336,0.0386],[338,0.0409],[342,0.0416],[344,0.0422],[350,0.0452],[354,0.0451],[358,0.0436],[360,0.0436],[364,0.0459],[366,0.0465],[368,0.0456],[370,0.0443],[372,0.044],[374,0.0451],[378,0.0456],[380,0.0445],[382,0.044],[384,0.0444],[388,0.0459],[390,0.0461],[392,0.0441],[394,0.0413],[396,0.0382],[398,0.0355],[400,0.0349],[402,0.0339],[404,0.0279],[408,0.0309],[410,0.0379],[412,0.0367],[414,0.0351],[416,0.0342],[418,0.0339],[422,0.0358],[424,0.036],[426,0.0365],[428,0.0358],[430,0.0361],[432,0.0356],[436,0.0365],[438,0.0365],[442,0.0358],[444,0.0361],[446,0.0361],[448,0.0354],[450,0.0345],[452,0.0342],[456,0.0341],[458,0.0332],[462,0.0319],[464,0.0322],[466,0.0322],[468,0.0316],[470,0.0302],[472,0.0274],[474,0.0243],[476,0.0219],[478,0.0242],[480,0.0295],[482,0.0336],[484,0.0342],[486,0.0295],[488,0.0271],[490,0.0261],[494,0.0327],[498,0.0418],[500,0.0456],[502,0.0462],[504,0.0473],[506,0.0491],[508,0.0518],[510,0.0524],[512,0.0523],[514,0.0511],[516,0.0469],[518,0.0467],[522,0.0411],[526,0.0404],[528,0.0409],[530,0.0428],[532,0.0432],[534,0.0434],[536,0.0431],[538,0.044],[540,0.0443],[546,0.0403],[548,0.0386],[550,0.038],[552,0.038],[554,0.0387],[556,0.0391],[558,0.0386],[560,0.0384],[562,0.0389],[564,0.0391],[566,0.0388],[568,0.0387],[570,0.0381],[572,0.0372],[576,0.0371],[580,0.0363],[582,0.0353],[584,0.0347],[586,0.0353],[588,0.0364],[590,0.0368],[592,0.0359],[594,0.0358],[596,0.0349],[598,0.0343],[600,0.0332],[602,0.0331],[604,0.0326],[608,0.0327],[610,0.0329],[612,0.0335],[614,0.0343],[616,0.0349],[620,0.0352],[622,0.0363],[626,0.0365],[628,0.0367],[636,0.0345],[638,0.0345],[640,0.0348],[642,0.0358],[644,0.0364],[646,0.0368],[648,0.0368],[650,0.0365],[652,0.0365],[656,0.0371],[662,0.0371],[664,0.0374],[670,0.0365],[672,0.0371],[674,0.0374],[676,0.0381],[678,0.0384],[680,0.039],[682,0.0381],[684,0.0378],[686,0.0378],[688,0.0384],[690,0.0387],[698,0.0388],[706,0.04],[710,0.0397],[712,0.0397],[716,0.0404],[720,0.0403]]
       }
       CTRL_R_mouth_corner.X {type=MovieSceneScriptingFloatChannel, default=0, const [0,492]}
       CTRL_R_mouth_corner.Y {
        type=MovieSceneScriptingFloatChannel
        default=0
        thinned_from=492
        keys [[-262,-0.0067],[-260,-0.0118],[-258,-0.0124],[-256,-0.0109],[-254,-0.0099],[-252,-0.0096],[-250,-0.0075],[-248,-0.0042],[-246,-0.0031],[-244,-0.0029],[-242,-0.003],[-240,-0.0016],[-238,-0.0009],[-236,-0.0021],[-234,-0.0021],[-232,-0.0018],[-230,-0.0021],[-226,-0.0055],[-220,-0.0092],[-218,-0.0101],[-214,-0.0145],[-212,-0.0149],[-210,-0.0161],[-208,-0.0171],[-206,-0.017],[-202,-0.0192],[-198,-0.0228],[-196,-0.0243],[-194,-0.0264],[-192,-0.0275],[-190,-0.0277],[-188,-0.0295],[-186,-0.0295],[-184,-0.028],[-182,-0.0259],[-180,-0.0217],[-178,-0.0135],[-176,-0.0026],[-174,0.0093],[-172,0.0187],[-168,0.0361],[-166,0.0441],[-164,0.0509],[-162,0.055],[-160,0.0544],[-158,0.054],[-156,0.0505],[-154,0.0442],[-152,0.0402],[-150,0.0365],[-148,0.0355],[-146,0.0355],[-144,0.0359],[-142,0.0374],[-140,0.0381],[-138,0.0394],[-136,0.0381],[-134,0.0411],[-132,0.0429],[-130,0.0467],[-128,0.0479],[-126,0.0506],[-124,0.0538],[-122,0.0573],[-120,0.0596],[-118,0.0561],[-116,0.0436],[-114,0.041],[-112,0.035],[-110,0.0313],[-108,0.0293],[-106,0.0283],[-104,0.0279],[-102,0.026],[-100,0.0249],[-98,0.0247],[-94,0.0293],[-92,0.0299],[-90,0.0341],[-88,0.0349],[-86,0.0374],[-84,0.0395],[-82,0.0414],[-78,0.0492],[-76,0.0501],[-72,0.0534],[-70,0.056],[-68,0.0589],[-66,0.0608],[-62,0.0572],[-60,0.045],[-58,0.043],[-56,0.0385],[-54,0.0364],[-52,0.0365],[-50,0.0368],[-48,0.0368],[-46,0.0355],[-44,0.0321],[-42,0.0284],[-40,0.0279],[-38,0.0254],[-36,0.0233],[-34,0.0189],[-32,0.0182],[-30,0.0179],[-28,0.0179],[-26,0.0176],[-24,0.0176],[-22,0.0182],[-20,0.0182],[-18,0.0174],[-16,0.0159],[-14,0.015],[-12,0.0153],[-10,0.0166],[-6,0.0172],[-2,0.0166],[0,0.0156],[2,0.0151],[6,0.0153],[8,0.015],[10,0.0137],[12,0.0137],[14,0.0152],[18,0.0155],[20,0.0148],[22,0.0138],[28,0.0127],[30,0.0133],[32,0.0137],[34,0.0137],[36,0.0122],[40,0.0117],[42,0.0137],[44,0.0146],[46,0.0149],[48,0.0157],[52,0.0166],[54,0.0168],[58,0.0178],[60,0.0177],[62,0.0173],[66,0.0176],[68,0.0173],[72,0.0173],[74,0.0177],[76,0.0178],[78,0.0173],[80,0.016],[82,0.0151],[84,0.0145],[86,0.0148],[88,0.0157],[90,0.0159],[92,0.0189],[94,0.0195],[96,0.0192],[100,0.0154],[102,0.0128],[104,0.0114],[106,0.0137],[108,0.0149],[110,0.0167],[112,0.0175],[114,0.0181],[116,0.0171],[118,0.015],[120,0.0147],[122,0.0135],[124,0.0156],[126,0.0157],[128,0.0169],[130,0.0169],[136,0.0179],[138,0.0179],[140,0.0183],[146,0.0211],[148,0.022],[150,0.0232],[152,0.025],[154,0.026],[156,0.026],[158,0.0257],[160,0.0251],[162,0.0251],[164,0.0254],[166,0.0254],[170,0.0247],[172,0.0251],[174,0.0244],[176,0.0244],[178,0.0242],[180,0.0247],[184,0.0254],[186,0.0253],[188,0.0248],[190,0.0241],[194,0.0234],[198,0.0217],[200,0.0217],[202,0.0223],[204,0.0224],[206,0.0216],[210,0.0215],[214,0.0221],[216,0.0222],[218,0.0228],[226,0.0241],[228,0.0241],[230,0.0244],[232,0.0244],[236,0.0257],[238,0.026],[240,0.0267],[242,0.0271],[244,0.028],[246,0.0276],[248,0.027],[250,0.0267],[252,0.0267],[256,0.0254],[258,0.0244],[262,0.0217],[264,0.0212],[266,0.0212],[270,0.0199],[272,0.0215],[274,0.0218],[276,0.0203],[278,0.0169],[280,0.0151],[282,0.0152],[284,0.0172],[286,0.0202],[288,0.0227],[290,0.0228],[292,0.0219],[294,0.0228],[296,0.0241],[298,0.0234],[302,0.0208],[304,0.0208],[306,0.0205],[308,0.0208],[310,0.0215],[312,0.0218],[314,0.0212],[316,0.0195],[318,0.0189],[320,0.0186],[322,0.0198],[324,0.0197],[326,0.0205],[328,0.021],[330,0.0225],[332,0.0229],[336,0.0273],[338,0.0313],[340,0.0319],[342,0.0312],[344,0.0302],[346,0.0296],[348,0.0279],[350,0.0257],[352,0.0229],[354,0.0225],[356,0.0211],[358,0.0193],[360,0.0182],[362,0.0162],[364,0.0156],[366,0.0154],[372,0.0141],[376,0.0143],[378,0.014],[382,0.0124],[386,0.009],[388,0.0054],[390,0.0052],[392,0.0011],[394,-0.0027],[396,-0.005],[398,-0.0028],[400,-0.0015],[402,0.004],[404,0.0025],[406,0.0006],[408,-0.0035],[410,-0.0001],[412,-0.0001],[414,-0.001],[418,-0.002],[420,-0.0003],[422,0.0017],[426,0.0029],[428,0.0023],[434,0.0029],[440,0.0039],[442,0.0045],[444,0.0049],[446,0.0046],[448,0.0032],[450,0.0023],[452,0.0026],[454,0.0033],[456,0.0045],[458,0.0042],[460,0.0052],[462,0.0059],[464,0.0068],[466,0.0057],[470,0.0029],[472,0.0022],[476,-0.0056],[478,-0.0067],[480,-0.0049],[482,-0.004],[484,-0.0039],[486,-0.007],[490,-0.0085],[492,-0.0109],[494,-0.0153],[496,-0.0182],[498,-0.0162],[500,-0.0114],[502,-0.0076],[504,-0.0015],[506,0.0065],[508,0.0152],[510,0.0201],[512,0.0205],[514,0.02],[516,0.0132],[518,0.0125],[520,0.0061],[522,0.0032],[524,0.0032],[528,0.0059],[530,0.01],[532,0.0106],[534,0.0087],[536,0.0041],[538,0.0006],[540,0.0002],[542,-0.0027],[544,-0.0005],[546,-0.001],[550,-0.0001],[552,0.0012],[554,0.0022],[556,0.0025],[560,0],[562,0.0009],[564,0.0016],[568,0.0016],[570,0],[574,-0.0019],[576,-0.0032],[578,-0.006],[580,-0.0075],[582,-0.0095],[584,-0.011],[586,-0.0128],[588,-0.0133],[590,-0.0129],[592,-0.012],[594,-0.0118],[596,-0.0088],[598,-0.0043],[600,0.0038],[602,0.0045],[604,0.0104],[606,0.0116],[608,0.0151],[610,0.0169],[612,0.0183],[614,0.0195],[616,0.0212],[618,0.0223],[620,0.0225],[622,0.0233],[624,0.0235],[626,0.0244],[628,0.0257],[630,0.0257],[632,0.0251],[634,0.0251],[638,0.0257],[640,0.0257],[644,0.027],[646,0.027],[648,0.0273],[652,0.0273],[654,0.0267],[656,0.0264],[662,0.0264],[664,0.0267],[666,0.0267],[668,0.0274],[670,0.0277],[672,0.0283],[676,0.029],[680,0.029],[682,0.0287],[684,0.0286],[688,0.028],[690,0.0283],[692,0.0283],[694,0.0286],[696,0.028],[698,0.0283],[700,0.0283],[702,0.029],[704,0.0287],[708,0.0295],[712,0.0296],[714,0.0293],[716,0.0293],[718,0.029],[720,0.029]]
       }
       CTRL_C_teeth_fwdBackD {type=MovieSceneScriptingFloatChannel, default=0, const [0,492]}
       CTRL_R_mouth_cornerSharpnessD {
        type=MovieSceneScriptingFloatChannel
        default=0
        thinned_from=492
        keys [[-262,0.0756],[-260,0.0743],[-256,0.0752],[-254,0.0774],[-252,0.0806],[-250,0.0847],[-248,0.0876],[-246,0.089],[-244,0.0894],[-242,0.0907],[-240,0.0881],[-238,0.086],[-236,0.0845],[-234,0.0841],[-232,0.0809],[-230,0.0782],[-228,0.0783],[-226,0.0782],[-224,0.0763],[-222,0.0734],[-220,0.072],[-218,0.0703],[-216,0.0646],[-214,0.0597],[-212,0.059],[-210,0.0572],[-208,0.0579],[-206,0.0591],[-204,0.059],[-202,0.0571],[-200,0.0548],[-198,0.0556],[-196,0.0527],[-194,0.052],[-192,0.053],[-190,0.0533],[-188,0.0552],[-186,0.0552],[-184,0.0557],[-182,0.0566],[-180,0.0599],[-178,0.0712],[-176,0.0855],[-174,0.0984],[-172,0.1023],[-170,0.1027],[-168,0.1037],[-166,0.1025],[-164,0.0966],[-162,0.0876],[-160,0.0805],[-158,0.0796],[-156,0.0707],[-154,0.0549],[-150,0.0411],[-148,0.0322],[-146,0.0317],[-144,0.0316],[-142,0.0298],[-138,0.0269],[-136,0.0283],[-134,0.0288],[-132,0.0288],[-130,0.0303],[-128,0.03],[-124,0.0262],[-122,0.0266],[-120,0.0245],[-118,0.0163],[-116,-0.0017],[-114,-0.0038],[-112,-0.0078],[-110,-0.0096],[-108,-0.0108],[-106,-0.0115],[-104,-0.0119],[-102,-0.0098],[-100,-0.0086],[-98,-0.0085],[-96,-0.0018],[-94,0.0014],[-92,0.0024],[-90,0.0087],[-88,0.0102],[-86,0.0156],[-84,0.0214],[-82,0.0264],[-80,0.0344],[-78,0.0401],[-76,0.041],[-74,0.0396],[-72,0.0349],[-70,0.0328],[-68,0.0337],[-66,0.029],[-64,0.0226],[-62,0.0201],[-60,0.0055],[-58,0.0045],[-54,0],[-52,-0.0015],[-50,-0.0033],[-48,-0.004],[-44,-0.0112],[-42,-0.0098],[-40,-0.0055],[-38,-0.0032],[-36,-0.0017],[-30,0.0005],[-28,0.001],[-26,0.001],[-24,0.0015],[-22,0.0015],[-20,0.0005],[-18,-0.0026],[-16,-0.004],[-14,-0.003],[-12,-0.0032],[-10,-0.0039],[-8,-0.0034],[-6,-0.0035],[-4,-0.0049],[-2,-0.0051],[0,-0.0049],[2,-0.0058],[6,-0.0054],[8,-0.0059],[10,-0.0054],[12,-0.0034],[14,-0.0025],[18,-0.0024],[20,-0.0028],[22,-0.0029],[24,-0.0036],[26,-0.0039],[28,-0.0032],[30,-0.002],[32,-0.0016],[34,-0.0017],[38,-0.0029],[40,-0.0013],[42,-0.0003],[44,-0.0009],[46,-0.0012],[48,-0.0019],[50,-0.0019],[52,-0.0017],[54,-0.0012],[56,-0.0011],[60,-0.002],[62,-0.0022],[66,-0.001],[70,-0.001],[72,-0.0013],[74,-0.0014],[78,0.0002],[80,-0.0012],[82,-0.0019],[84,-0.0015],[86,-0.0017],[88,-0.0024],[90,-0.0027],[92,-0.0041],[94,-0.0041],[100,-0.0022],[104,-0.0015],[106,-0.0015],[108,-0.001],[110,0.0008],[112,0.0033],[114,0.0051],[118,0.0076],[120,0.0078],[124,0.0088],[126,0.0088],[128,0.0093],[130,0.0093],[134,0.0074],[136,0.0078],[138,0.0069],[140,0.0054],[142,0.0054],[144,0.0049],[146,0.0052],[150,0.0063],[152,0.0062],[154,0.0039],[156,0.0038],[158,0.0024],[160,0.0029],[162,0.0039],[164,0.0034],[166,0.002],[168,0.0024],[170,0.0024],[172,0.0015],[174,0.001],[176,0.0015],[178,0.0014],[180,0.001],[182,0.0024],[184,0.0015],[188,0.0012],[190,0.0021],[192,0.0024],[194,0.0024],[196,0.0027],[198,0.0024],[200,0.002],[202,0.0026],[204,0.0029],[206,0.0016],[208,0.0019],[210,0.002],[212,0.0029],[214,0.0029],[216,0.0035],[220,0.0053],[222,0.0068],[226,0.0078],[228,0.0088],[230,0.0088],[232,0.0083],[234,0.0089],[236,0.0107],[238,0.0108],[240,0.0112],[242,0.0127],[246,0.0118],[248,0.0122],[252,0.0093],[254,0.0097],[256,0.0097],[258,0.0083],[260,0.0049],[262,0.0041],[264,0.0053],[266,0.0059],[268,0.0074],[272,0.0112],[274,0.0112],[276,0.0054],[278,-0.0015],[280,-0.0059],[282,-0.0068],[284,-0.0061],[286,-0.0031],[288,0.0018],[290,0.0034],[292,0.0025],[294,0.0029],[296,0.0024],[300,0.0054],[302,0.0033],[304,0.002],[306,-0.0002],[308,-0.0002],[310,0.0019],[312,0.0024],[314,0.0015],[318,-0.0012],[320,-0.0013],[322,0.0004],[324,0],[326,0],[328,0.0013],[330,0.001],[332,-0.0002],[334,0.0002],[336,0.0048],[338,0.0106],[340,0.0114],[342,0.012],[344,0.0137],[346,0.015],[350,0.0185],[354,0.0177],[356,0.0165],[358,0.0159],[360,0.0156],[362,0.0142],[364,0.0142],[368,0.0133],[370,0.0132],[372,0.0136],[374,0.0124],[376,0.0122],[378,0.0117],[380,0.0122],[382,0.0122],[384,0.0137],[386,0.0164],[388,0.0195],[390,0.0202],[392,0.023],[394,0.0256],[396,0.027],[398,0.028],[400,0.0283],[402,0.0267],[404,0.0286],[406,0.0257],[408,0.0183],[410,0.0171],[412,0.017],[416,0.0161],[418,0.0155],[420,0.0137],[422,0.0132],[424,0.0132],[428,0.0127],[430,0.0127],[432,0.0123],[434,0.0122],[436,0.0127],[438,0.0113],[440,0.0108],[442,0.0122],[444,0.0122],[448,0.0103],[450,0.0103],[452,0.0112],[454,0.0112],[456,0.0118],[458,0.0127],[460,0.0127],[464,0.0137],[466,0.0151],[468,0.0151],[470,0.0138],[472,0.019],[474,0.0191],[476,0.0159],[478,0.0144],[480,0.0124],[482,0.0117],[484,0.0117],[486,0.0122],[488,0.0122],[490,0.0128],[492,0.0199],[494,0.041],[496,0.0733],[498,0.095],[500,0.0896],[502,0.0813],[504,0.0711],[506,0.0655],[510,0.0664],[512,0.0637],[514,0.0678],[516,0.0733],[518,0.0738],[520,0.0778],[522,0.0766],[524,0.0731],[526,0.0659],[528,0.0637],[530,0.06],[532,0.059],[536,0.0503],[538,0.0525],[540,0.0532],[542,0.0572],[544,0.0576],[546,0.0588],[548,0.0628],[550,0.0655],[552,0.0673],[554,0.0687],[556,0.0684],[558,0.0662],[560,0.0659],[562,0.0691],[564,0.0695],[566,0.0676],[568,0.0674],[570,0.0675],[572,0.0686],[574,0.0657],[576,0.0607],[578,0.0548],[582,0.0539],[584,0.0528],[586,0.0532],[588,0.054],[590,0.0534],[592,0.0524],[594,0.0518],[596,0.0411],[598,0.03],[600,0.0191],[602,0.0186],[604,0.0151],[606,0.0147],[608,0.0133],[610,0.0126],[612,0.012],[614,0.012],[616,0.0122],[622,0.0123],[624,0.0122],[626,0.0117],[630,0.0117],[632,0.0112],[638,0.0112],[640,0.0122],[642,0.0127],[646,0.0127],[648,0.0122],[650,0.0122],[652,0.0127],[654,0.0127],[656,0.0122],[658,0.0123],[660,0.0132],[662,0.0132],[664,0.0127],[668,0.0137],[670,0.0137],[672,0.0127],[674,0.0122],[676,0.0132],[678,0.0132],[682,0.0142],[684,0.0137],[686,0.0142],[688,0.0142],[690,0.0146],[692,0.0146],[694,0.0142],[696,0.0146],[698,0.0142],[702,0.0142],[706,0.0132],[708,0.0132],[710,0.0128],[714,0.0127],[716,0.0122],[718,0.0112],[720,0.0112]]
       }
       CTRL_R_mouth_cornerSharpnessU {
        type=MovieSceneScriptingFloatChannel
        default=0
        thinned_from=492
        keys [[-262,0.1187],[-258,0.1197],[-256,0.1181],[-252,0.1215],[-250,0.1221],[-248,0.1222],[-246,0.1217],[-244,0.1216],[-242,0.1229],[-240,0.1199],[-238,0.1164],[-236,0.1152],[-234,0.1147],[-232,0.1105],[-230,0.1078],[-228,0.1093],[-226,0.1094],[-224,0.106],[-222,0.1029],[-220,0.1021],[-218,0.1021],[-216,0.1004],[-214,0.0959],[-212,0.0951],[-210,0.0924],[-208,0.0954],[-206,0.0987],[-204,0.0991],[-202,0.0981],[-200,0.0953],[-198,0.0971],[-194,0.0943],[-192,0.0957],[-190,0.0963],[-188,0.1005],[-186,0.1004],[-182,0.0955],[-180,0.0958],[-178,0.1001],[-176,0.1037],[-174,0.104],[-172,0.0978],[-170,0.0885],[-168,0.0795],[-166,0.0712],[-162,0.0487],[-160,0.042],[-158,0.0415],[-156,0.0373],[-154,0.029],[-152,0.0258],[-150,0.0249],[-148,0.0215],[-146,0.0215],[-144,0.0229],[-142,0.0229],[-140,0.0225],[-138,0.0234],[-136,0.0239],[-134,0.021],[-132,0.021],[-130,0.024],[-128,0.0239],[-126,0.0203],[-124,0.017],[-122,0.0165],[-120,0.0167],[-118,0.0187],[-116,0.0137],[-114,0.012],[-112,0.0082],[-110,0.0048],[-108,0.0032],[-106,0.0037],[-104,0.0039],[-100,0.0067],[-98,0.0068],[-96,0.0108],[-94,0.0127],[-92,0.013],[-90,0.0142],[-88,0.014],[-86,0.0137],[-84,0.0145],[-82,0.0158],[-80,0.0158],[-78,0.0136],[-76,0.0127],[-74,0.0093],[-72,0.0089],[-70,0.0117],[-68,0.0122],[-66,0.0123],[-64,0.0146],[-62,0.0148],[-60,0.0137],[-58,0.0131],[-56,0.0116],[-54,0.0086],[-50,0.0015],[-48,0.0005],[-46,0.0043],[-44,0.0099],[-42,0.0138],[-40,0.017],[-38,0.0157],[-36,0.0122],[-34,0.0112],[-32,0.0112],[-30,0.0098],[-28,0.0092],[-24,0.0063],[-22,0.0063],[-20,0.0056],[-18,0.0028],[-14,0.0012],[-12,0],[-10,-0.0015],[-8,-0.0027],[-6,-0.0034],[-4,-0.0039],[-2,-0.0027],[0,-0.0024],[2,-0.0032],[4,-0.0027],[6,-0.0019],[8,-0.0015],[10,-0.0012],[12,-0.0002],[14,0],[20,0],[24,-0.0005],[26,-0.0003],[32,0.0014],[34,0.0023],[38,0.002],[40,0.0024],[42,0.0014],[44,0],[46,-0.0003],[48,-0.001],[50,-0.001],[52,-0.0008],[54,-0.0003],[56,-0.0005],[60,0],[62,0.0004],[64,0.0013],[66,0.001],[70,0.001],[72,0.0014],[74,0.001],[76,0.0011],[82,0.0026],[84,0.0027],[86,0.0022],[88,0.0012],[90,0.0007],[92,-0.0003],[94,0.0022],[96,0.0066],[98,0.0088],[100,0.0116],[102,0.0146],[104,0.0165],[106,0.0161],[108,0.0163],[110,0.0172],[114,0.0221],[116,0.0238],[118,0.026],[120,0.0264],[122,0.0263],[124,0.0259],[126,0.0259],[128,0.0264],[132,0.0264],[136,0.0254],[138,0.0244],[142,0.0234],[144,0.022],[146,0.0218],[150,0.022],[152,0.0205],[154,0.0176],[156,0.0175],[160,0.0157],[162,0.0156],[164,0.0152],[166,0.0137],[168,0.0151],[170,0.0151],[172,0.0132],[176,0.0142],[178,0.0129],[180,0.0122],[182,0.0132],[184,0.0137],[188,0.0133],[190,0.0137],[194,0.0139],[196,0.0151],[198,0.0151],[200,0.0141],[202,0.014],[204,0.0142],[206,0.0147],[208,0.0137],[210,0.0137],[212,0.0148],[214,0.0151],[216,0.0152],[218,0.0161],[222,0.02],[228,0.02],[232,0.022],[234,0.022],[236,0.0234],[238,0.0238],[240,0.0235],[242,0.0234],[246,0.0225],[248,0.0229],[250,0.022],[252,0.0205],[256,0.0215],[258,0.0205],[260,0.0186],[262,0.0186],[264,0.0195],[266,0.02],[268,0.022],[270,0.0229],[272,0.0225],[274,0.0239],[276,0.0179],[278,0.0078],[280,0.0013],[282,-0.0004],[284,-0.0012],[286,-0.0002],[288,0.0033],[290,0.0053],[292,0.0039],[294,0.0034],[296,0.0034],[298,0.0049],[300,0.0059],[302,0.0043],[304,0.0034],[306,0.0019],[308,0.0015],[310,0.0033],[312,0.0044],[316,0.0053],[318,0.0039],[322,0.0039],[324,0.0036],[328,0.0054],[332,0.0055],[334,0.0073],[336,0.0111],[338,0.017],[340,0.0191],[342,0.0206],[346,0.0251],[348,0.0275],[350,0.0293],[352,0.0298],[354,0.0297],[356,0.0293],[360,0.0293],[362,0.0281],[364,0.0278],[366,0.0282],[370,0.03],[372,0.0307],[374,0.0303],[378,0.0303],[380,0.0323],[382,0.0332],[384,0.0368],[386,0.045],[388,0.0526],[390,0.0525],[392,0.0531],[394,0.0531],[398,0.0476],[400,0.0469],[402,0.0402],[404,0.0449],[406,0.0461],[408,0.035],[410,0.0335],[414,0.0371],[416,0.0381],[418,0.037],[420,0.0347],[422,0.0337],[424,0.0337],[426,0.0334],[428,0.0337],[430,0.0337],[432,0.0328],[434,0.0327],[436,0.0332],[440,0.0313],[442,0.0312],[446,0.0322],[448,0.0317],[450,0.0317],[452,0.0322],[456,0.0323],[458,0.0332],[462,0.0322],[464,0.0336],[468,0.0395],[470,0.0454],[472,0.0602],[474,0.0678],[476,0.0662],[478,0.0629],[480,0.061],[482,0.0598],[484,0.0596],[486,0.0579],[488,0.0542],[490,0.0501],[492,0.0594],[494,0.0825],[496,0.1107],[498,0.1241],[500,0.1171],[502,0.1129],[506,0.0924],[508,0.0893],[510,0.0889],[512,0.0877],[514,0.0927],[516,0.102],[518,0.1021],[520,0.0989],[522,0.0889],[526,0.0766],[528,0.074],[530,0.0709],[532,0.0729],[534,0.0811],[536,0.0942],[538,0.1054],[540,0.1069],[542,0.1065],[544,0.0983],[546,0.0958],[548,0.0969],[550,0.0995],[552,0.1017],[554,0.1019],[556,0.1011],[558,0.1001],[560,0.1003],[562,0.1019],[564,0.1013],[566,0.1002],[568,0.1001],[570,0.0997],[572,0.1013],[574,0.0997],[576,0.0966],[578,0.0924],[580,0.0923],[582,0.0943],[584,0.0999],[586,0.1106],[588,0.1152],[590,0.1124],[592,0.1081],[594,0.1066],[596,0.0867],[600,0.0495],[602,0.048],[604,0.0361],[606,0.0347],[608,0.0304],[610,0.0285],[612,0.0274],[618,0.026],[620,0.0259],[622,0.0251],[624,0.0249],[626,0.0244],[628,0.0244],[630,0.0249],[632,0.0251],[634,0.0244],[636,0.0234],[638,0.0234],[640,0.0239],[644,0.0239],[646,0.0244],[652,0.023],[656,0.0239],[660,0.0239],[662,0.0234],[664,0.0234],[668,0.0242],[670,0.0235],[672,0.0225],[676,0.0234],[678,0.023],[682,0.0229],[684,0.0225],[686,0.0234],[688,0.0239],[692,0.0239],[694,0.0234],[696,0.0239],[700,0.0239],[704,0.023],[706,0.0222],[714,0.0224],[716,0.0215],[720,0.0215]]
       }
       CTRL_L_eye_eyelidU.X {type=MovieSceneScriptingFloatChannel, default=0, const [0,492]}
       CTRL_L_eye_eyelidU.Y {type=MovieSceneScriptingFloatChannel, default=0, const [0,492]}
       CTRL_L_mouth_cornerSharpnessD {
        type=MovieSceneScriptingFloatChannel
        default=0
        thinned_from=492
        keys [[-262,0.0931],[-260,0.0915],[-258,0.0921],[-256,0.0932],[-254,0.0955],[-252,0.0985],[-250,0.102],[-248,0.1051],[-246,0.107],[-244,0.1074],[-242,0.1088],[-240,0.1066],[-238,0.1041],[-236,0.1021],[-234,0.1017],[-232,0.0989],[-230,0.0966],[-228,0.0971],[-226,0.0958],[-224,0.0935],[-222,0.0928],[-220,0.0913],[-218,0.0882],[-216,0.0834],[-214,0.0796],[-212,0.079],[-210,0.0761],[-208,0.0745],[-206,0.0765],[-204,0.0775],[-202,0.0747],[-200,0.0742],[-198,0.0761],[-196,0.073],[-194,0.0732],[-192,0.0744],[-190,0.0748],[-188,0.0771],[-186,0.0773],[-182,0.0802],[-180,0.0837],[-178,0.0939],[-176,0.1053],[-174,0.1132],[-172,0.114],[-168,0.1129],[-166,0.1101],[-164,0.1013],[-162,0.0907],[-160,0.084],[-158,0.083],[-156,0.0741],[-154,0.0591],[-152,0.053],[-150,0.0479],[-148,0.0415],[-144,0.0404],[-142,0.0391],[-140,0.0386],[-138,0.0376],[-136,0.039],[-132,0.0381],[-130,0.0386],[-128,0.0388],[-126,0.0395],[-124,0.0382],[-122,0.0382],[-120,0.0355],[-118,0.0261],[-116,0.0068],[-114,0.0052],[-112,0.0022],[-108,-0.0011],[-106,-0.0018],[-104,-0.0022],[-102,-0.0004],[-100,0.0009],[-98,0.001],[-96,0.0074],[-94,0.0107],[-92,0.0121],[-90,0.0209],[-88,0.0227],[-84,0.0355],[-82,0.0404],[-80,0.0473],[-78,0.0516],[-76,0.0522],[-74,0.0508],[-72,0.047],[-70,0.0445],[-68,0.0444],[-66,0.0398],[-64,0.0338],[-62,0.0313],[-60,0.0162],[-58,0.0149],[-56,0.0121],[-54,0.01],[-52,0.0083],[-50,0.0057],[-48,0.0048],[-44,-0.0017],[-42,-0.0017],[-40,0.005],[-38,0.0097],[-36,0.0113],[-34,0.0127],[-32,0.0132],[-28,0.0151],[-26,0.0146],[-22,0.0127],[-20,0.0119],[-16,0.0084],[-12,0.0083],[-8,0.0064],[-6,0.0058],[-4,0.0044],[-2,0.0039],[0,0.0044],[2,0.0032],[4,0.0024],[8,0.0025],[10,0.0039],[12,0.0059],[14,0.0063],[16,0.0063],[18,0.0061],[20,0.0054],[22,0.0052],[24,0.0047],[26,0.0044],[28,0.0047],[32,0.0076],[34,0.0082],[38,0.0064],[40,0.0077],[42,0.0083],[44,0.0074],[46,0.0071],[48,0.0064],[50,0.0065],[54,0.0075],[56,0.0075],[58,0.007],[62,0.0073],[64,0.0069],[68,0.0078],[70,0.0078],[74,0.0066],[76,0.007],[78,0.0071],[80,0.0062],[82,0.0059],[86,0.0059],[90,0.0054],[92,0.0034],[94,0.0034],[96,0.0053],[98,0.0064],[100,0.0077],[102,0.0081],[104,0.0092],[106,0.0112],[108,0.0117],[110,0.0134],[112,0.0162],[114,0.0179],[116,0.0184],[118,0.0197],[120,0.02],[122,0.0211],[124,0.0229],[130,0.0229],[132,0.0224],[134,0.0215],[136,0.022],[138,0.0215],[140,0.0195],[142,0.0191],[144,0.0181],[146,0.0189],[148,0.0187],[150,0.0194],[152,0.0194],[154,0.0166],[156,0.0165],[158,0.0151],[160,0.0147],[162,0.0156],[164,0.0152],[166,0.0122],[168,0.0136],[170,0.0141],[172,0.0118],[174,0.0117],[176,0.0122],[180,0.0108],[182,0.0122],[188,0.0111],[190,0.0123],[192,0.0127],[194,0.0127],[196,0.0129],[198,0.0129],[200,0.0127],[204,0.0132],[206,0.0123],[210,0.0122],[212,0.0133],[218,0.0147],[222,0.0176],[224,0.0171],[228,0.0181],[232,0.0181],[234,0.0186],[236,0.0205],[238,0.0205],[240,0.02],[242,0.02],[244,0.019],[248,0.02],[250,0.0182],[252,0.0166],[256,0.0166],[258,0.0152],[260,0.0122],[262,0.0114],[264,0.0122],[266,0.0127],[268,0.0148],[270,0.018],[272,0.022],[274,0.022],[276,0.0132],[278,0.0058],[280,0.0026],[282,0.0015],[284,0.0024],[286,0.0061],[288,0.0101],[292,0.0113],[294,0.0122],[296,0.0117],[300,0.0137],[302,0.0116],[306,0.0088],[308,0.0088],[310,0.0111],[312,0.0117],[314,0.0098],[318,0.0078],[320,0.0079],[322,0.0092],[324,0.0084],[326,0.0093],[328,0.011],[332,0.0093],[334,0.0103],[336,0.0151],[338,0.0218],[340,0.0237],[344,0.0279],[346,0.0284],[350,0.0351],[354,0.0342],[356,0.0326],[358,0.0324],[360,0.0327],[362,0.0322],[364,0.032],[368,0.0309],[372,0.0307],[374,0.0299],[378,0.0298],[384,0.0311],[386,0.0305],[388,0.0308],[390,0.0322],[392,0.036],[394,0.0391],[396,0.0407],[400,0.0405],[402,0.0371],[404,0.0353],[406,0.0346],[408,0.0332],[410,0.0322],[412,0.0325],[414,0.0332],[418,0.0312],[422,0.0283],[424,0.0286],[426,0.0291],[428,0.0288],[432,0.0288],[434,0.0283],[436,0.0288],[438,0.0283],[440,0.0273],[444,0.0264],[446,0.0264],[450,0.0254],[454,0.0264],[458,0.0283],[460,0.0283],[462,0.0298],[464,0.0307],[466,0.0323],[468,0.0327],[470,0.0304],[472,0.033],[474,0.0305],[476,0.0257],[478,0.0274],[480,0.0295],[482,0.031],[484,0.0312],[486,0.0312],[488,0.0298],[490,0.0313],[492,0.0408],[494,0.0649],[496,0.0996],[498,0.1219],[500,0.1155],[502,0.1071],[504,0.0969],[506,0.0904],[510,0.0903],[512,0.0867],[516,0.0957],[518,0.0962],[520,0.0999],[522,0.0975],[524,0.0916],[526,0.0825],[528,0.0809],[530,0.077],[532,0.0757],[534,0.072],[536,0.0671],[538,0.0679],[540,0.0684],[542,0.0704],[544,0.0718],[546,0.0741],[548,0.0794],[550,0.0835],[552,0.0864],[554,0.0877],[556,0.0877],[558,0.0864],[560,0.0869],[562,0.0887],[564,0.0885],[566,0.0867],[568,0.0865],[570,0.0871],[572,0.0892],[574,0.0864],[576,0.0818],[578,0.0762],[580,0.0761],[582,0.0754],[584,0.0724],[586,0.0716],[590,0.0768],[592,0.0763],[594,0.0756],[596,0.0621],[598,0.0508],[600,0.0377],[602,0.037],[604,0.0317],[606,0.031],[608,0.0288],[610,0.0275],[612,0.0266],[614,0.0266],[616,0.0269],[622,0.027],[624,0.0268],[626,0.0259],[632,0.0244],[634,0.0244],[636,0.024],[638,0.0244],[640,0.0258],[642,0.0264],[646,0.0254],[654,0.0273],[656,0.0269],[658,0.0269],[660,0.0273],[664,0.0273],[666,0.0269],[668,0.027],[670,0.0273],[672,0.0269],[674,0.0269],[676,0.0273],[678,0.0273],[680,0.0278],[682,0.0278],[684,0.0274],[686,0.0283],[688,0.0288],[692,0.0288],[694,0.0283],[696,0.0288],[698,0.0289],[702,0.0298],[704,0.0288],[706,0.0283],[708,0.0283],[710,0.028],[712,0.0278],[716,0.0269],[720,0.0269]]
       }
       CTRL_L_eye_eyelidD.X {type=MovieSceneScriptingFloatChannel, default=0, const [0,492]}
       CTRL_L_eye_eyelidD.Y {type=MovieSceneScriptingFloatChannel, default=0, const [0,492]}
       CTRL_L_mouth_cornerSharpnessU {
        type=MovieSceneScriptingFloatChannel
        default=0
        thinned_from=492
        keys [[-262,0.085],[-260,0.0858],[-258,0.088],[-256,0.0885],[-254,0.0899],[-250,0.0939],[-248,0.0955],[-244,0.0952],[-242,0.0961],[-240,0.0931],[-238,0.0895],[-236,0.0879],[-234,0.0874],[-232,0.0839],[-230,0.0817],[-226,0.0826],[-224,0.081],[-222,0.0789],[-218,0.0775],[-216,0.0755],[-214,0.0718],[-212,0.0712],[-210,0.0684],[-208,0.0691],[-206,0.0705],[-204,0.0713],[-202,0.0713],[-200,0.0694],[-198,0.0713],[-194,0.0681],[-192,0.0697],[-190,0.0704],[-188,0.0727],[-186,0.0725],[-184,0.07],[-182,0.0667],[-180,0.067],[-176,0.0761],[-174,0.0784],[-172,0.0744],[-170,0.0688],[-168,0.0625],[-166,0.0546],[-164,0.0425],[-162,0.0307],[-160,0.0264],[-156,0.0264],[-154,0.0259],[-152,0.0297],[-150,0.0375],[-148,0.0375],[-146,0.0361],[-142,0.0391],[-140,0.0409],[-138,0.0415],[-136,0.0406],[-134,0.0391],[-132,0.0395],[-130,0.0418],[-128,0.0432],[-126,0.041],[-124,0.0379],[-122,0.036],[-120,0.0307],[-118,0.0232],[-116,0.0103],[-114,0.009],[-112,0.0063],[-110,0.004],[-108,0.003],[-106,0.004],[-104,0.0044],[-102,0.0043],[-100,0.0053],[-98,0.0054],[-96,0.0093],[-94,0.0102],[-92,0.01],[-90,0.0083],[-88,0.0083],[-86,0.0077],[-84,0.0061],[-82,0.0057],[-80,0.0049],[-76,0.0054],[-74,0.0205],[-72,0.0278],[-70,0.0317],[-68,0.0317],[-66,0.027],[-64,0.0192],[-62,0.0176],[-60,0.0108],[-58,0.01],[-52,0.0054],[-50,0.0032],[-48,0.0024],[-46,0.0029],[-44,0.0064],[-42,0.0106],[-40,0.0106],[-38,0.005],[-36,0.001],[-34,0.001],[-32,0.0015],[-30,0.0015],[-28,0.0009],[-26,-0.0007],[-24,-0.0009],[-22,-0.0002],[-20,-0.0008],[-18,-0.0036],[-16,-0.0048],[-14,-0.0047],[-10,-0.0056],[-6,-0.0071],[-4,-0.0066],[-2,-0.0054],[0,-0.0059],[2,-0.0074],[4,-0.0078],[6,-0.0073],[8,-0.0076],[10,-0.0083],[12,-0.0088],[14,-0.0088],[16,-0.009],[18,-0.0096],[20,-0.0098],[22,-0.0094],[26,-0.0092],[28,-0.0089],[30,-0.0083],[32,-0.0081],[34,-0.0081],[36,-0.0085],[38,-0.0083],[40,-0.0076],[42,-0.0073],[44,-0.008],[46,-0.0084],[48,-0.0096],[50,-0.0101],[52,-0.0099],[54,-0.0089],[56,-0.0082],[60,-0.0084],[62,-0.0077],[64,-0.0076],[66,-0.0078],[70,-0.0074],[72,-0.0064],[74,-0.0066],[76,-0.0065],[78,-0.006],[80,-0.0066],[82,-0.0062],[84,-0.0053],[86,-0.0051],[88,-0.0056],[92,-0.0062],[94,-0.0041],[96,-0.0008],[98,0.0003],[100,0.0023],[102,0.0052],[104,0.0068],[106,0.0074],[108,0.0085],[112,0.0139],[114,0.0158],[118,0.0179],[122,0.0181],[124,0.0195],[130,0.0195],[132,0.019],[142,0.019],[144,0.0181],[146,0.0179],[150,0.0181],[152,0.018],[154,0.0161],[156,0.0161],[158,0.0151],[164,0.0166],[166,0.0156],[168,0.0156],[170,0.0166],[172,0.0161],[174,0.0152],[176,0.0156],[178,0.0156],[180,0.0151],[188,0.015],[190,0.0161],[194,0.0157],[196,0.0159],[200,0.0142],[202,0.0148],[204,0.0151],[206,0.0147],[208,0.0156],[210,0.0156],[212,0.0171],[216,0.0171],[218,0.0176],[222,0.0205],[224,0.02],[226,0.02],[230,0.022],[232,0.022],[234,0.021],[236,0.0205],[240,0.0215],[242,0.0225],[244,0.0229],[246,0.0215],[248,0.0215],[250,0.022],[252,0.0215],[254,0.0215],[256,0.021],[260,0.0181],[262,0.0185],[266,0.0215],[268,0.0219],[270,0.021],[272,0.0211],[274,0.0229],[276,0.0205],[278,0.0131],[280,0.0076],[282,0.0071],[284,0.0073],[286,0.0078],[288,0.0107],[290,0.0122],[292,0.0108],[294,0.0103],[296,0.0103],[300,0.0141],[302,0.0126],[304,0.0117],[306,0.0102],[308,0.0093],[310,0.0107],[316,0.0137],[320,0.0117],[322,0.0117],[324,0.0113],[328,0.0122],[330,0.0122],[334,0.0142],[336,0.0163],[338,0.0198],[340,0.0199],[342,0.0195],[344,0.0207],[346,0.0214],[348,0.0219],[350,0.0229],[352,0.0234],[354,0.0232],[358,0.0212],[360,0.0205],[362,0.019],[364,0.0185],[366,0.0178],[368,0.0191],[370,0.0199],[372,0.0192],[374,0.0174],[378,0.0166],[380,0.0176],[382,0.0176],[384,0.0187],[386,0.0247],[388,0.0349],[390,0.0354],[392,0.0344],[396,0.03],[398,0.0293],[400,0.0293],[402,0.0268],[404,0.0317],[406,0.0285],[408,0.016],[410,0.0139],[412,0.0151],[414,0.0166],[416,0.0171],[418,0.0165],[420,0.0151],[426,0.0152],[428,0.0156],[430,0.0152],[432,0.0143],[434,0.0146],[436,0.0156],[440,0.0147],[442,0.0151],[444,0.0151],[446,0.0142],[448,0.0137],[452,0.0146],[456,0.0147],[458,0.0156],[460,0.0147],[462,0.0142],[464,0.0151],[466,0.0158],[468,0.0171],[470,0.0199],[472,0.0329],[474,0.0397],[476,0.0388],[478,0.0344],[480,0.0303],[482,0.0277],[484,0.0273],[486,0.0263],[488,0.0249],[490,0.0222],[492,0.028],[494,0.0453],[496,0.0676],[498,0.0778],[500,0.0727],[502,0.0713],[504,0.0662],[506,0.0606],[508,0.0581],[510,0.0576],[512,0.0563],[514,0.0588],[516,0.063],[518,0.063],[520,0.0612],[522,0.0549],[526,0.0468],[530,0.0443],[532,0.0463],[534,0.0533],[536,0.0634],[538,0.0721],[540,0.0732],[542,0.0735],[544,0.0688],[546,0.0669],[548,0.0674],[552,0.0716],[554,0.0727],[556,0.0735],[558,0.0728],[560,0.0729],[562,0.0746],[564,0.0744],[566,0.0725],[568,0.0723],[570,0.0723],[572,0.0736],[574,0.0713],[578,0.0642],[580,0.064],[582,0.065],[584,0.0687],[586,0.0758],[588,0.0791],[590,0.0779],[592,0.0739],[594,0.0725],[596,0.0538],[598,0.0393],[600,0.0265],[602,0.0258],[604,0.0205],[606,0.0198],[608,0.0179],[610,0.0176],[614,0.0176],[618,0.0181],[620,0.018],[622,0.0172],[624,0.0171],[626,0.0166],[628,0.0166],[632,0.0174],[634,0.0171],[636,0.0171],[638,0.0181],[642,0.0171],[644,0.018],[646,0.0181],[648,0.0176],[650,0.0181],[652,0.0176],[654,0.0176],[656,0.0181],[658,0.018],[662,0.0171],[664,0.0171],[666,0.0176],[668,0.0184],[670,0.0176],[672,0.0166],[674,0.0166],[676,0.0171],[682,0.0171],[684,0.0166],[688,0.0176],[694,0.0176],[696,0.0181],[698,0.018],[704,0.0166],[706,0.0159],[712,0.0161],[714,0.0156],[716,0.0156],[718,0.0161],[720,0.0161]]
       }
       CTRL_R_mouth_lipsTowardsTeethD {type=MovieSceneScriptingFloatChannel, default=0, const [0,492]}
       CTRL_L_mouth_lipsTowardsTeethD {type=MovieSceneScriptingFloatChannel, default=0, const [0,492]}
       CTRL_R_eye_eyelidU.X {type=MovieSceneScriptingFloatChannel, default=0, const [0,492]}
       CTRL_R_eye_eyelidU.Y {type=MovieSceneScriptingFloatChannel, default=0, const [0,492]}
       CTRL_L_mouth_lipsTowardsTeethU {type=MovieSceneScriptingFloatChannel, default=0, const [0,492]}
       CTRL_R_eye_eyelidD.X {type=MovieSceneScriptingFloatChannel, default=0, const [0,492]}
       CTRL_R_eye_eyelidD.Y {type=MovieSceneScriptingFloatChannel, default=0, const [0,492]}
       CTRL_R_mouth_lipsTowardsTeethU {type=MovieSceneScriptingFloatChannel, default=0, const [0,492]}
       CTRL_faceGUIfollowHead {type=MovieSceneScriptingFloatChannel, default=1, const [1,492]}
       CTRL_eyesAimFollowHead {type=MovieSceneScriptingFloatChannel, default=0, const [0,492]}
       CTRL_C_nose.X {type=MovieSceneScriptingFloatChannel, default=0, const [0,492]}
       CTRL_C_nose.Y {type=MovieSceneScriptingFloatChannel, default=0, const [0,492]}
       CTRL_C_teeth_rotateU {type=MovieSceneScriptingFloatChannel, default=0, const [0,492]}
       CTRL_C_teeth_rotateD {type=MovieSceneScriptingFloatChannel, default=0, const [0,492]}
       CTRL_L_brow_inner {type=MovieSceneScriptingFloatChannel, default=0, const [0,492]}
       CTRL_R_brow_inner {type=MovieSceneScriptingFloatChannel, default=0, const [0,492]}
       CTRL_L_brow_outer {type=MovieSceneScriptingFloatChannel, default=0, const [0,492]}
       CTRL_R_brow_outer {type=MovieSceneScriptingFloatChannel, default=0, const [0,492]}
       CTRL_C_mouth.X {
        type=MovieSceneScriptingFloatChannel
        default=0
        thinned_from=492
        keys [[-262,0.0414],[-258,0.0411],[-256,0.0406],[-254,0.0396],[-252,0.0394],[-250,0.0387],[-248,0.0391],[-246,0.0384],[-244,0.0381],[-242,0.0407],[-240,0.0424],[-238,0.0432],[-234,0.0431],[-232,0.0433],[-228,0.0453],[-226,0.0454],[-224,0.0458],[-222,0.0468],[-216,0.0477],[-214,0.0487],[-212,0.0488],[-210,0.0479],[-208,0.0494],[-206,0.0489],[-204,0.048],[-202,0.0479],[-200,0.0483],[-198,0.0469],[-196,0.0462],[-194,0.0463],[-192,0.0458],[-190,0.0456],[-186,0.0459],[-184,0.0477],[-182,0.0489],[-180,0.048],[-178,0.0465],[-176,0.0464],[-174,0.0456],[-170,0.0417],[-164,0.0366],[-162,0.0382],[-160,0.0403],[-158,0.04],[-156,0.0373],[-154,0.0371],[-152,0.0352],[-150,0.0362],[-148,0.0355],[-146,0.0329],[-144,0.0343],[-142,0.0369],[-140,0.0385],[-138,0.0357],[-136,0.0355],[-134,0.0322],[-132,0.0312],[-130,0.0294],[-126,0.0331],[-124,0.0306],[-122,0.0286],[-120,0.026],[-118,0.0194],[-116,0.0102],[-114,0.0094],[-112,0.0072],[-110,0.0045],[-108,0.0043],[-106,0.0051],[-104,0.0054],[-102,0.0065],[-100,0.0059],[-98,0.0058],[-96,0.0072],[-94,0.008],[-92,0.008],[-90,0.0083],[-88,0.0084],[-86,0.0089],[-84,0.0104],[-82,0.0124],[-80,0.0158],[-78,0.0208],[-76,0.0225],[-74,0.0318],[-72,0.0369],[-70,0.0329],[-68,0.0273],[-66,0.0247],[-64,0.0199],[-62,0.0182],[-60,0.0099],[-58,0.0092],[-56,0.0075],[-54,0.0061],[-52,0.007],[-50,0.0078],[-48,0.0078],[-46,0.0103],[-44,0.0106],[-42,0.0114],[-40,0.0083],[-38,0.0096],[-36,0.0097],[-32,0.0125],[-30,0.0121],[-28,0.0124],[-26,0.0121],[-24,0.0114],[-22,0.012],[-20,0.0115],[-18,0.0096],[-16,0.0093],[-14,0.0101],[-12,0.0104],[-8,0.0119],[-6,0.0117],[-4,0.0098],[-2,0.0107],[0,0.0098],[2,0.0106],[4,0.0113],[6,0.0115],[8,0.0121],[10,0.012],[12,0.013],[14,0.0148],[16,0.0147],[18,0.0143],[20,0.0145],[24,0.0142],[26,0.0138],[28,0.014],[30,0.0147],[32,0.0143],[34,0.0131],[36,0.0128],[38,0.0132],[40,0.013],[44,0.0115],[48,0.0118],[50,0.0108],[52,0.0101],[54,0.01],[56,0.0111],[58,0.0115],[60,0.0123],[64,0.0113],[68,0.0116],[70,0.0115],[72,0.0113],[74,0.0119],[76,0.0113],[78,0.0104],[80,0.0112],[82,0.0116],[84,0.0109],[86,0.0097],[88,0.0092],[90,0.0093],[92,0.0111],[94,0.0101],[96,0.0111],[98,0.0114],[100,0.0129],[102,0.0146],[104,0.0154],[108,0.0152],[110,0.0167],[114,0.0202],[116,0.021],[118,0.0221],[120,0.0223],[122,0.024],[124,0.0244],[126,0.0246],[128,0.0268],[130,0.0268],[134,0.0252],[136,0.0279],[138,0.0266],[140,0.0255],[142,0.0255],[144,0.0233],[146,0.0228],[148,0.0218],[150,0.0204],[152,0.0188],[154,0.0181],[156,0.0181],[160,0.0185],[162,0.019],[164,0.0193],[166,0.0181],[168,0.0192],[172,0.0195],[174,0.0174],[176,0.018],[178,0.017],[180,0.0172],[182,0.0169],[184,0.0176],[186,0.0173],[188,0.0167],[190,0.0165],[192,0.0152],[194,0.0149],[196,0.015],[198,0.0147],[200,0.0146],[202,0.0152],[206,0.0152],[210,0.0154],[212,0.0173],[214,0.0174],[216,0.018],[218,0.0178],[220,0.0172],[224,0.0194],[226,0.0244],[228,0.0259],[230,0.0251],[232,0.0253],[234,0.026],[236,0.0302],[238,0.0322],[240,0.0333],[246,0.0331],[248,0.0306],[250,0.0305],[252,0.029],[254,0.0313],[256,0.0323],[260,0.0322],[262,0.0314],[264,0.0322],[266,0.0327],[268,0.0353],[270,0.0426],[272,0.0479],[274,0.0538],[276,0.0462],[278,0.0303],[280,0.0225],[282,0.0205],[284,0.0197],[286,0.0228],[288,0.025],[292,0.0253],[294,0.0258],[296,0.0267],[298,0.0266],[300,0.0281],[304,0.0263],[306,0.0243],[308,0.0248],[310,0.0261],[312,0.0258],[314,0.0244],[318,0.0244],[320,0.024],[322,0.0249],[324,0.0228],[326,0.0235],[328,0.0226],[330,0.0236],[332,0.0224],[334,0.0215],[336,0.0345],[338,0.0353],[342,0.0354],[346,0.0359],[348,0.0366],[350,0.037],[354,0.0368],[356,0.0335],[358,0.0313],[360,0.0309],[362,0.0313],[364,0.032],[366,0.0317],[368,0.0328],[374,0.039],[376,0.0394],[378,0.0429],[380,0.0471],[382,0.0505],[384,0.0566],[386,0.0634],[388,0.0663],[390,0.0622],[392,0.0551],[394,0.052],[396,0.0522],[398,0.0499],[400,0.0487],[402,0.0454],[404,0.0543],[406,0.046],[408,0.0428],[410,0.0447],[412,0.044],[414,0.0439],[418,0.0409],[420,0.0369],[422,0.0359],[426,0.0376],[428,0.0392],[430,0.0383],[432,0.0358],[434,0.0364],[436,0.0372],[438,0.0351],[440,0.0346],[442,0.0354],[444,0.0342],[446,0.0327],[448,0.0325],[450,0.0333],[454,0.0332],[456,0.0344],[458,0.0372],[460,0.0381],[462,0.0393],[464,0.0419],[466,0.0457],[468,0.0449],[470,0.0407],[472,0.0435],[474,0.0429],[476,0.0386],[478,0.0324],[480,0.0289],[482,0.0283],[484,0.0283],[486,0.0328],[490,0.0391],[492,0.0461],[494,0.0559],[496,0.0624],[498,0.0663],[500,0.0646],[502,0.0627],[504,0.0619],[506,0.0594],[508,0.0581],[510,0.0598],[512,0.0574],[514,0.0568],[516,0.0597],[518,0.0588],[520,0.0493],[522,0.0421],[524,0.0359],[526,0.0318],[528,0.031],[530,0.0296],[534,0.0294],[536,0.0308],[538,0.0313],[540,0.0313],[544,0.0293],[546,0.0315],[548,0.0346],[550,0.038],[552,0.0399],[554,0.0414],[556,0.0424],[558,0.0436],[562,0.0444],[564,0.0442],[566,0.0432],[568,0.0431],[572,0.045],[576,0.0436],[584,0.0435],[586,0.045],[588,0.0458],[590,0.0455],[594,0.0456],[596,0.0454],[598,0.0437],[600,0.0443],[602,0.0438],[604,0.0389],[606,0.0379],[608,0.0347],[610,0.032],[612,0.0288],[614,0.0269],[616,0.0269],[618,0.0275],[620,0.0276],[622,0.0241],[624,0.0235],[626,0.0245],[628,0.0244],[630,0.024],[632,0.0245],[634,0.024],[638,0.0243],[640,0.0234],[642,0.0243],[644,0.0243],[646,0.0249],[650,0.0233],[652,0.0228],[654,0.0222],[656,0.0218],[660,0.0204],[662,0.0213],[664,0.0209],[666,0.0202],[668,0.0213],[670,0.0214],[672,0.0203],[674,0.0204],[676,0.0199],[678,0.0202],[682,0.0193],[684,0.0194],[686,0.0203],[688,0.0199],[690,0.0197],[694,0.0216],[696,0.022],[698,0.0213],[700,0.0214],[702,0.0209],[704,0.0212],[706,0.0212],[708,0.0216],[710,0.0212],[712,0.021],[714,0.0201],[716,0.0189],[718,0.0192],[720,0.0192]]
       }
       CTRL_C_mouth.Y {type=MovieSceneScriptingFloatChannel, default=0, const [0,492]}
       CTRL_L_mouth_upperLipRaise {
        type=MovieSceneScriptingFloatChannel
        default=0
        thinned_from=492
        keys [[-262,0.0026],[-260,0.0012],[-258,0.0012],[-252,0],[-250,0.0003],[-248,0.0004],[-242,0],[-240,0.0007],[-238,0.0003],[-236,0.0007],[-234,0.0008],[-232,0.0012],[-230,0.0013],[-226,0.0009],[-222,0.0009],[-220,0.0005],[-218,0.0006],[-214,0.0013],[-212,0.0013],[-210,0.0011],[-208,0.0013],[-206,0.0011],[-202,0.001],[-200,0.0005],[-190,0.0011],[-188,0.0008],[-184,0.0007],[-182,0.0004],[-180,0.001],[-176,0.0016],[-174,0.0022],[-172,0.0025],[-168,0.0028],[-166,0.0031],[-158,0.0032],[-154,0.0027],[-152,0.0028],[-148,0.0026],[-146,0.003],[-144,0.0029],[-142,0.0026],[-136,0.0028],[-134,0.0031],[-132,0.003],[-130,0.0026],[-118,0.0028],[-108,0.0023],[-98,0.0026],[-92,0.0027],[-90,0.003],[-76,0.0029],[-62,0.0029],[-60,0.0025],[-48,0.0022],[-46,0.0019],[-44,0.0026],[-40,0.003],[-38,0.0028],[-34,0.0029],[-32,0.0033],[-28,0.0032],[-26,0.0033],[-22,0.0032],[-20,0.0033],[-16,0.0032],[-12,0.0034],[-10,0.0032],[-8,0.0033],[-6,0.0032],[-2,0.0034],[0,0.0032],[4,0.0034],[6,0.0033],[8,0.0035],[12,0.0031],[32,0.0032],[40,0.003],[42,0.0031],[44,0.0028],[50,0.0031],[56,0.0029],[60,0.0032],[66,0.0028],[82,0.0029],[86,0.0031],[90,0.003],[92,0.0033],[102,0.0025],[104,0.0026],[106,0.0024],[110,0.0026],[112,0.0024],[116,0.0007],[118,0.0011],[124,0.0017],[130,0.0016],[132,0.0022],[134,0.0017],[136,0.0016],[138,0.002],[144,0.0017],[146,0.0014],[148,0.0014],[150,0.0018],[152,0.0015],[156,0.0016],[158,0.001],[162,0.001],[164,0.0007],[168,0.0006],[170,0.0007],[174,0.0015],[176,0.0012],[178,0.0013],[182,0.0009],[194,0.0009],[196,0.0005],[200,0.0015],[206,0.0012],[208,0.0009],[212,0.0007],[214,0.001],[216,0.0003],[220,0.0008],[222,0.0009],[226,0.0007],[228,0.0003],[230,0.0009],[232,0.0011],[234,0.0007],[236,0.0007],[238,0.0004],[240,0.0008],[244,0.0009],[246,0.0005],[248,0.0006],[250,0.0003],[254,0.0004],[256,0.0001],[258,0.0002],[260,0.0006],[264,0],[268,0.0011],[272,0.001],[274,0.0036],[276,0.0177],[278,0.0147],[280,0.0205],[282,0.019],[286,0.0145],[288,0.0081],[290,0.0057],[292,0.0074],[294,0.0058],[296,0.0074],[298,0.0041],[300,0.0038],[302,0.0066],[304,0.0071],[306,0.0106],[308,0.0102],[310,0.008],[312,0.0086],[314,0.0081],[318,0.0117],[320,0.0106],[322,0.011],[324,0.0109],[326,0.0094],[328,0.0081],[330,0.0077],[332,0.0075],[334,0.0053],[336,0.002],[344,0.002],[350,0.0018],[354,0.0018],[356,0.0017],[358,0.0008],[360,0.0009],[362,0.0013],[364,0.0012],[366,0.0008],[368,0.0007],[370,0.0011],[376,0.0014],[380,0.0009],[382,0.0009],[384,0.0005],[388,0.0027],[390,0.0027],[392,0.0024],[394,0.0018],[396,0.0026],[398,0.0032],[400,0.0032],[402,0.0029],[404,0.0036],[406,0.0032],[414,0.0026],[416,0.0026],[420,0.0025],[422,0.0028],[426,0.0024],[432,0.0026],[434,0.0022],[438,0.0024],[440,0.0029],[444,0.0027],[446,0.0029],[448,0.0028],[454,0.0031],[456,0.0026],[462,0.0025],[466,0.0028],[468,0.0026],[470,0.0013],[472,0.0024],[474,0.0027],[476,0.0026],[478,0.0015],[480,0.0015],[484,0.0017],[486,0.001],[488,0.0008],[490,0.0012],[492,0.0006],[494,0.0006],[496,0.0002],[498,0.0006],[500,0.0002],[502,0.0007],[506,0.0003],[508,0.0011],[512,0.0006],[514,0],[518,0.0002],[520,0.0015],[522,0.0022],[524,0.002],[526,0.0029],[530,0.0033],[536,0.0031],[538,0.0026],[540,0.0025],[542,0.0026],[548,0.001],[552,0.0005],[558,0.0005],[562,0.0011],[564,0.0008],[568,0.001],[570,0.0009],[572,0.0002],[576,0.0006],[578,0.0007],[582,0.0002],[586,0.0005],[588,0.0005],[592,0.0011],[596,0.0012],[600,0.0024],[602,0.0024],[604,0.002],[614,0.0027],[616,0.0024],[626,0.0022],[630,0.002],[632,0.002],[636,0.0018],[638,0.0019],[640,0.0023],[642,0.0022],[644,0.0018],[646,0.0016],[648,0.0022],[652,0.0024],[654,0.002],[662,0.0023],[668,0.002],[672,0.0022],[674,0.0019],[676,0.0021],[678,0.0019],[680,0.0021],[684,0.0018],[686,0.002],[688,0.0018],[690,0.0019],[696,0.0017],[698,0.0018],[700,0.0016],[702,0.002],[704,0.002],[706,0.0023],[720,0.0019]]
       }
       CTRL_L_mouth_cornerPull {
        type=MovieSceneScriptingFloatChannel
        default=0.5472
        thinned_from=492
        keys [[-262,0.0018],[-258,0.0012],[-256,0.0005],[-252,0],[-244,0.0002],[-242,0],[-240,0.0004],[-232,0.0004],[-228,0.0007],[-222,0.0007],[-218,0.0002],[-216,0.0004],[-214,0.0008],[-210,0.0011],[-198,0.0003],[-190,0.0009],[-184,0.0006],[-182,0.0003],[-170,0.0023],[-164,0.0028],[-158,0.0028],[-154,0.0026],[-150,0.0026],[-146,0.0023],[-120,0.0022],[-118,0.0028],[-116,0.0072],[-114,0.0097],[-112,0.0167],[-110,0.0242],[-108,0.029],[-106,0.0316],[-102,0.0339],[-100,0.0326],[-98,0.0325],[-96,0.0319],[-94,0.0302],[-92,0.0297],[-88,0.0293],[-86,0.0277],[-84,0.0215],[-82,0.0155],[-80,0.0104],[-78,0.0069],[-76,0.0061],[-74,0.0029],[-68,0.0025],[-66,0.0026],[-64,0.0042],[-62,0.0067],[-60,0.0239],[-58,0.0269],[-56,0.0338],[-54,0.0385],[-52,0.044],[-50,0.0498],[-48,0.0514],[-46,0.0478],[-44,0.04],[-42,0.0285],[-40,0.0276],[-38,0.0284],[-36,0.0328],[-34,0.032],[-32,0.0337],[-30,0.0346],[-28,0.0372],[-26,0.0384],[-24,0.0401],[-22,0.0413],[-20,0.0407],[-18,0.0414],[-16,0.0425],[-14,0.0449],[-12,0.0452],[-10,0.0482],[-8,0.0503],[-6,0.0533],[-4,0.0553],[0,0.0561],[2,0.0546],[6,0.0563],[8,0.0581],[10,0.0568],[12,0.0553],[14,0.0564],[18,0.0568],[20,0.0562],[22,0.0562],[24,0.0557],[26,0.0549],[28,0.0532],[30,0.0529],[32,0.0529],[34,0.0518],[38,0.0482],[42,0.0502],[44,0.0518],[46,0.0521],[48,0.0528],[50,0.0531],[54,0.0524],[56,0.0526],[58,0.0536],[60,0.0543],[62,0.0546],[66,0.0526],[70,0.0527],[72,0.0533],[74,0.0537],[76,0.0534],[78,0.0527],[80,0.0508],[82,0.0494],[84,0.0495],[86,0.0505],[88,0.0512],[92,0.0514],[94,0.0498],[96,0.0434],[100,0.0343],[102,0.0294],[104,0.0262],[106,0.0272],[108,0.0311],[110,0.0395],[112,0.0442],[114,0.0465],[116,0.0462],[118,0.0489],[120,0.0497],[122,0.0509],[124,0.0571],[126,0.0576],[128,0.0649],[130,0.0649],[132,0.0685],[134,0.0679],[136,0.0652],[138,0.0657],[140,0.0675],[142,0.0706],[144,0.0724],[150,0.083],[152,0.0876],[154,0.0942],[156,0.0945],[158,0.0994],[160,0.1076],[162,0.1114],[164,0.1193],[166,0.1211],[168,0.1195],[170,0.1164],[172,0.119],[174,0.1205],[176,0.1208],[178,0.116],[180,0.1176],[182,0.1157],[184,0.1108],[186,0.111],[190,0.113],[192,0.1131],[194,0.1119],[196,0.1075],[198,0.1068],[202,0.112],[204,0.1128],[206,0.11],[208,0.1118],[210,0.1119],[212,0.1125],[214,0.1114],[216,0.1125],[218,0.1112],[220,0.107],[222,0.1013],[224,0.1023],[226,0.1024],[228,0.1051],[230,0.1046],[232,0.1103],[234,0.1126],[236,0.1109],[238,0.1049],[240,0.1002],[242,0.1093],[244,0.1212],[246,0.1276],[248,0.1283],[250,0.1273],[252,0.1315],[256,0.1344],[258,0.139],[260,0.1456],[262,0.1484],[264,0.1517],[266,0.1595],[270,0.1711],[274,0.1803],[276,0.192],[278,0.2031],[280,0.2135],[282,0.2157],[284,0.221],[286,0.2221],[288,0.221],[290,0.216],[292,0.2148],[294,0.2144],[296,0.2167],[298,0.2147],[302,0.2162],[304,0.2173],[308,0.2167],[310,0.2159],[314,0.2153],[316,0.2156],[320,0.2117],[322,0.2112],[324,0.212],[326,0.2133],[328,0.212],[330,0.2117],[334,0.2069],[336,0.1523],[338,0.1488],[340,0.1469],[344,0.1414],[348,0.1345],[350,0.1308],[352,0.128],[354,0.1276],[356,0.1203],[358,0.1152],[360,0.1142],[362,0.1086],[364,0.1049],[366,0.1013],[368,0.1003],[370,0.1],[374,0.0953],[376,0.0948],[378,0.0891],[382,0.077],[384,0.0707],[386,0.0683],[388,0.0657],[390,0.0652],[392,0.0577],[394,0.0476],[396,0.0371],[398,0.0331],[400,0.033],[402,0.0331],[404,0.0351],[406,0.0348],[408,0.035],[410,0.029],[412,0.027],[414,0.0256],[416,0.0259],[418,0.0255],[420,0.0277],[422,0.0309],[424,0.0327],[426,0.0349],[428,0.0343],[432,0.0358],[434,0.0377],[438,0.039],[440,0.0382],[442,0.0369],[448,0.0316],[450,0.0295],[452,0.0287],[454,0.0274],[456,0.0255],[458,0.0238],[460,0.0229],[464,0.0227],[466,0.0212],[468,0.0199],[470,0.0176],[472,0.0168],[474,0.0151],[476,0.0122],[478,0.0088],[480,0.0062],[482,0.004],[484,0.0037],[486,0.0014],[488,0.0008],[492,0.0005],[494,0.0005],[496,0.0002],[498,0.0004],[500,0.0002],[502,0.0004],[504,0.0004],[508,0.0038],[510,0.0058],[512,0.0054],[516,0.0019],[518,0.0018],[520,0.001],[524,0.0017],[526,0.0029],[528,0.0037],[530,0.0054],[532,0.0063],[536,0.0048],[538,0.0038],[542,0.0033],[544,0.0021],[548,0.0007],[550,0.0003],[552,0],[560,0],[570,0.0002],[584,0],[588,0.0002],[592,0.0008],[594,0.0008],[596,0.0013],[598,0.0019],[600,0.0044],[602,0.0051],[604,0.0124],[606,0.0142],[608,0.0199],[610,0.0231],[612,0.0253],[616,0.0269],[620,0.0267],[622,0.0263],[624,0.0262],[626,0.0267],[628,0.0287],[630,0.029],[634,0.0286],[636,0.0297],[638,0.0303],[640,0.0304],[642,0.0301],[644,0.0292],[646,0.0277],[648,0.0269],[650,0.0277],[652,0.0276],[656,0.0259],[658,0.0257],[660,0.0253],[662,0.0246],[670,0.0249],[674,0.0261],[676,0.0265],[678,0.0261],[680,0.0261],[682,0.0267],[684,0.0279],[686,0.0285],[688,0.0281],[692,0.0285],[696,0.0282],[698,0.0296],[700,0.0312],[702,0.0335],[710,0.0385],[716,0.0394],[720,0.0394]]
       }
       CTRL_L_mouth_dimple {
        type=MovieSceneScriptingFloatChannel
        default=0
        thinned_from=492
        keys [[-262,0.0337],[-260,0.0266],[-258,0.0204],[-256,0.018],[-254,0.0194],[-252,0.0211],[-248,0.0263],[-246,0.0283],[-242,0.0292],[-240,0.0295],[-238,0.0295],[-236,0.0304],[-234,0.0305],[-232,0.0315],[-230,0.0313],[-228,0.0294],[-224,0.0228],[-222,0.0199],[-220,0.0183],[-216,0.0143],[-214,0.0127],[-212,0.0125],[-210,0.0115],[-208,0.0107],[-206,0.011],[-204,0.0106],[-202,0.009],[-200,0.0089],[-198,0.007],[-196,0.0061],[-192,0.0035],[-190,0.0032],[-188,0.0024],[-186,0.0023],[-182,0.003],[-180,0.0054],[-178,0.0118],[-176,0.0202],[-172,0.0409],[-170,0.0507],[-168,0.063],[-166,0.0757],[-164,0.0914],[-162,0.1038],[-160,0.12],[-158,0.1218],[-156,0.1377],[-152,0.1836],[-150,0.2077],[-148,0.2344],[-146,0.2576],[-142,0.2843],[-140,0.2865],[-138,0.2842],[-136,0.2848],[-134,0.2795],[-130,0.2803],[-128,0.2843],[-126,0.2859],[-124,0.2888],[-122,0.2892],[-120,0.2824],[-118,0.2668],[-116,0.2311],[-114,0.2218],[-112,0.2002],[-110,0.1853],[-108,0.1747],[-106,0.1687],[-104,0.1662],[-102,0.1509],[-100,0.1409],[-98,0.14],[-96,0.1231],[-94,0.1237],[-92,0.1249],[-90,0.1347],[-88,0.1381],[-86,0.1502],[-84,0.1604],[-82,0.1685],[-80,0.1823],[-78,0.1985],[-76,0.2034],[-74,0.2479],[-72,0.2751],[-70,0.3043],[-68,0.3175],[-66,0.3158],[-64,0.2981],[-62,0.2903],[-60,0.24],[-58,0.2297],[-56,0.2055],[-54,0.1902],[-52,0.1842],[-50,0.1836],[-48,0.1835],[-46,0.1718],[-44,0.1565],[-42,0.138],[-40,0.1222],[-38,0.1131],[-36,0.1068],[-34,0.1054],[-32,0.1006],[-30,0.0999],[-28,0.0987],[-26,0.0983],[-24,0.0974],[-22,0.0967],[-20,0.0987],[-18,0.0971],[-16,0.0968],[-14,0.098],[-12,0.0961],[-10,0.0982],[-8,0.0986],[-6,0.0969],[-4,0.0979],[-2,0.0949],[0,0.0938],[2,0.092],[4,0.089],[6,0.0884],[8,0.087],[10,0.0873],[12,0.0869],[14,0.0874],[16,0.0872],[20,0.0856],[24,0.0856],[28,0.0861],[32,0.0854],[34,0.0854],[38,0.0837],[40,0.083],[42,0.0839],[44,0.0852],[46,0.0855],[48,0.0869],[50,0.0892],[52,0.09],[54,0.0918],[56,0.0923],[58,0.0921],[62,0.0912],[66,0.0915],[68,0.0911],[70,0.0908],[72,0.0892],[74,0.0889],[76,0.0892],[78,0.09],[80,0.09],[82,0.0885],[84,0.0868],[86,0.0859],[88,0.0871],[90,0.0875],[92,0.0872],[94,0.0882],[96,0.0861],[98,0.0858],[100,0.0851],[102,0.0824],[104,0.0815],[106,0.0943],[108,0.1013],[110,0.1189],[112,0.1323],[114,0.1429],[116,0.1457],[118,0.145],[120,0.1448],[122,0.1468],[124,0.1506],[126,0.151],[128,0.1575],[130,0.1575],[132,0.159],[134,0.158],[136,0.1553],[138,0.152],[140,0.147],[142,0.1451],[144,0.1451],[146,0.1497],[148,0.1577],[150,0.166],[152,0.1776],[154,0.1848],[156,0.185],[158,0.1884],[160,0.1907],[162,0.1917],[164,0.1923],[166,0.1915],[168,0.1915],[170,0.1886],[172,0.1888],[174,0.1879],[176,0.1856],[178,0.1839],[180,0.1817],[182,0.1802],[184,0.1774],[186,0.1764],[188,0.1733],[190,0.1713],[192,0.1707],[196,0.1688],[200,0.1661],[202,0.1651],[204,0.1646],[206,0.1633],[208,0.1648],[210,0.1649],[212,0.1667],[214,0.1674],[218,0.1703],[220,0.1738],[222,0.1813],[224,0.1857],[228,0.1847],[230,0.1856],[232,0.19],[236,0.2105],[238,0.2175],[244,0.2296],[246,0.2309],[248,0.2339],[250,0.2326],[252,0.2295],[254,0.2281],[256,0.225],[258,0.2242],[260,0.2187],[262,0.2169],[264,0.2158],[266,0.2159],[268,0.2232],[270,0.2331],[272,0.2471],[274,0.2563],[276,0.2619],[280,0.2794],[282,0.2886],[284,0.3159],[286,0.3208],[288,0.3215],[290,0.3137],[292,0.3048],[294,0.3033],[300,0.3111],[302,0.3106],[304,0.3085],[306,0.3083],[308,0.3091],[310,0.3065],[312,0.3044],[314,0.3033],[316,0.2981],[318,0.2965],[320,0.2911],[322,0.2897],[324,0.2913],[326,0.2922],[328,0.2918],[330,0.2905],[332,0.2904],[334,0.2888],[336,0.2231],[340,0.22],[344,0.2174],[350,0.212],[352,0.211],[354,0.2107],[356,0.2061],[358,0.2024],[360,0.1999],[362,0.1922],[366,0.1778],[368,0.1738],[370,0.1703],[372,0.166],[374,0.1588],[376,0.1575],[378,0.1503],[380,0.1402],[382,0.1383],[384,0.137],[386,0.1385],[388,0.1465],[390,0.1489],[392,0.1496],[394,0.1404],[396,0.1294],[398,0.1211],[400,0.1188],[402,0.1261],[404,0.1196],[406,0.1235],[408,0.1254],[410,0.1196],[412,0.1276],[414,0.125],[416,0.1204],[418,0.1184],[420,0.1192],[422,0.1211],[424,0.1208],[426,0.1216],[428,0.1194],[430,0.1183],[432,0.1189],[434,0.117],[436,0.118],[438,0.1183],[440,0.1163],[442,0.1196],[446,0.1159],[448,0.1133],[450,0.1088],[452,0.108],[454,0.107],[456,0.1075],[458,0.107],[460,0.1086],[462,0.1094],[464,0.1078],[466,0.1073],[468,0.1035],[470,0.1],[472,0.0945],[474,0.0848],[476,0.0769],[478,0.0673],[480,0.0635],[482,0.0589],[484,0.0579],[486,0.0527],[488,0.0479],[490,0.0426],[492,0.0334],[494,0.0263],[496,0.0207],[498,0.0173],[500,0.0192],[502,0.0242],[504,0.0344],[506,0.0459],[508,0.0558],[510,0.0621],[512,0.0617],[514,0.0547],[516,0.0444],[518,0.0437],[520,0.0365],[522,0.0303],[524,0.0271],[526,0.0336],[528,0.0369],[530,0.0455],[532,0.0519],[534,0.0535],[538,0.0432],[540,0.0425],[542,0.0392],[544,0.0368],[546,0.0357],[548,0.032],[550,0.0285],[552,0.0278],[556,0.0301],[558,0.0308],[562,0.0315],[566,0.0318],[568,0.0318],[570,0.0311],[572,0.0294],[574,0.0291],[576,0.0285],[578,0.0269],[580,0.0261],[582,0.0243],[584,0.0219],[586,0.0209],[588,0.0207],[590,0.0228],[592,0.0246],[594,0.025],[596,0.0294],[598,0.0347],[600,0.0443],[602,0.0457],[604,0.0575],[606,0.0596],[608,0.0674],[610,0.0746],[612,0.0798],[614,0.0841],[616,0.0871],[618,0.0895],[620,0.0903],[622,0.0895],[624,0.0893],[626,0.0885],[630,0.088],[632,0.0883],[636,0.0901],[638,0.0911],[640,0.0932],[642,0.0943],[644,0.0944],[648,0.0918],[650,0.0907],[652,0.0906],[658,0.0929],[662,0.0946],[664,0.0963],[666,0.0972],[668,0.0974],[670,0.0988],[672,0.0996],[674,0.1009],[676,0.1027],[680,0.1045],[682,0.1044],[686,0.1056],[688,0.1052],[690,0.1057],[692,0.105],[694,0.1033],[696,0.1036],[698,0.1033],[700,0.104],[702,0.1055],[704,0.1044],[706,0.1049],[708,0.1042],[710,0.1038],[712,0.1038],[714,0.1024],[716,0.1004],[718,0.0978],[720,0.0978]]
       }
       CTRL_L_mouth_cornerDepress {
        type=MovieSceneScriptingFloatChannel
        default=0
        thinned_from=492
        keys [[-262,0.0118],[-260,0.0142],[-258,0.0147],[-256,0.0134],[-254,0.0124],[-252,0.0118],[-250,0.0107],[-248,0.0089],[-246,0.0081],[-242,0.0077],[-240,0.0079],[-238,0.0079],[-236,0.0075],[-232,0.0074],[-230,0.0077],[-226,0.0092],[-224,0.0104],[-222,0.0119],[-220,0.0125],[-218,0.0128],[-216,0.0158],[-214,0.0204],[-212,0.021],[-210,0.0219],[-208,0.0234],[-206,0.0237],[-204,0.0264],[-202,0.0282],[-200,0.0303],[-198,0.0316],[-196,0.034],[-194,0.0382],[-192,0.0409],[-190,0.0415],[-188,0.0453],[-186,0.0452],[-184,0.0424],[-182,0.0356],[-180,0.0282],[-178,0.0196],[-176,0.0126],[-174,0.0086],[-172,0.0078],[-170,0.008],[-168,0.0079],[-162,0.0069],[-158,0.0069],[-156,0.0071],[-154,0.0083],[-152,0.0098],[-150,0.0135],[-148,0.0178],[-146,0.0204],[-142,0.0221],[-140,0.0226],[-138,0.0234],[-136,0.0238],[-134,0.0219],[-132,0.0203],[-130,0.0184],[-128,0.0192],[-126,0.0181],[-122,0.0141],[-120,0.0134],[-118,0.0145],[-114,0.0122],[-112,0.0098],[-108,0.0066],[-106,0.006],[-104,0.0059],[-102,0.0061],[-100,0.0057],[-98,0.0057],[-96,0.005],[-94,0.0048],[-88,0.0048],[-80,0.0059],[-76,0.006],[-72,0.0106],[-70,0.012],[-68,0.0122],[-64,0.0093],[-62,0.0088],[-60,0.0062],[-58,0.0058],[-56,0.0049],[-54,0.0045],[-48,0.0042],[-42,0.0054],[-40,0.0054],[-32,0.0053],[-30,0.005],[-12,0.0054],[-8,0.0052],[-4,0.0055],[2,0.0052],[16,0.0051],[22,0.0048],[38,0.0051],[42,0.0048],[44,0.0044],[56,0.0043],[72,0.0041],[90,0.0046],[98,0.0044],[100,0.0046],[104,0.0057],[108,0.0052],[110,0.0044],[116,0.0023],[118,0.002],[120,0.002],[124,0.0026],[126,0.0026],[128,0.003],[130,0.003],[134,0.0033],[142,0.0031],[146,0.0026],[156,0.0027],[158,0.0024],[166,0.0022],[168,0.0019],[172,0.002],[174,0.0023],[180,0.0019],[194,0.0017],[196,0.0016],[202,0.0024],[204,0.0024],[206,0.0017],[208,0.0014],[210,0.0014],[216,0.0019],[220,0.0019],[224,0.002],[228,0.0016],[236,0.002],[238,0.002],[240,0.0023],[244,0.0024],[246,0.0021],[248,0.002],[252,0.0013],[254,0.0011],[258,0.0016],[262,0.0015],[264,0.0012],[266,0.0013],[268,0.0023],[270,0.0037],[272,0.0041],[276,0.007],[278,0.0064],[280,0.0049],[282,0.0035],[284,0.003],[286,0.0031],[288,0.0036],[290,0.0035],[292,0.003],[300,0.0028],[302,0.0031],[306,0.0034],[308,0.0033],[310,0.0036],[316,0.0029],[318,0.0032],[320,0.0031],[324,0.0031],[326,0.0034],[332,0.0033],[334,0.0035],[336,0.0034],[344,0.0036],[350,0.0033],[354,0.0034],[356,0.0038],[362,0.0031],[364,0.0035],[366,0.0036],[368,0.0033],[370,0.0032],[372,0.0035],[374,0.0039],[382,0.0043],[384,0.0049],[386,0.0063],[388,0.0081],[390,0.008],[392,0.0082],[394,0.0087],[396,0.0096],[398,0.0098],[400,0.0097],[402,0.0092],[404,0.0186],[406,0.0196],[408,0.0201],[410,0.009],[414,0.0099],[418,0.0102],[422,0.0089],[426,0.0085],[428,0.0087],[432,0.0081],[440,0.0077],[442,0.0075],[446,0.0075],[448,0.0082],[450,0.0085],[452,0.0087],[456,0.0082],[458,0.0085],[462,0.0078],[464,0.0071],[466,0.0073],[468,0.0078],[470,0.0078],[476,0.0093],[480,0.0081],[484,0.0081],[486,0.0088],[488,0.0088],[490,0.0086],[492,0.0089],[494,0.0105],[496,0.0113],[498,0.0114],[500,0.0098],[502,0.0089],[504,0.0074],[506,0.0061],[508,0.0052],[510,0.0049],[512,0.0048],[514,0.0041],[516,0.0046],[518,0.0047],[520,0.0064],[522,0.0076],[524,0.0085],[526,0.0089],[534,0.0086],[536,0.0091],[540,0.0093],[542,0.0092],[544,0.0085],[546,0.0084],[556,0.0069],[560,0.0072],[564,0.0068],[570,0.0072],[572,0.0066],[574,0.0069],[576,0.0075],[578,0.0091],[580,0.0096],[586,0.0101],[590,0.0101],[592,0.0108],[594,0.0109],[596,0.0106],[598,0.0098],[600,0.0082],[602,0.008],[604,0.0065],[606,0.0062],[608,0.005],[610,0.0045],[612,0.0042],[622,0.0034],[638,0.0029],[642,0.0033],[646,0.0029],[648,0.0029],[652,0.0037],[658,0.0036],[664,0.004],[680,0.0039],[682,0.0037],[690,0.0037],[692,0.0036],[694,0.0037],[700,0.0033],[704,0.0033],[706,0.0036],[716,0.0035],[720,0.0033]]
       }
       CTRL_L_mouth_stretch {
        type=MovieSceneScriptingFloatChannel
        default=0
        thinned_from=492
        keys [[-262,0.0516],[-260,0.0526],[-258,0.0502],[-256,0.0488],[-254,0.05],[-252,0.0516],[-248,0.0517],[-246,0.0509],[-244,0.0506],[-242,0.0525],[-238,0.0552],[-236,0.0561],[-234,0.0564],[-232,0.0585],[-230,0.0601],[-228,0.0608],[-226,0.0595],[-224,0.0593],[-222,0.0604],[-220,0.0613],[-218,0.0597],[-216,0.0569],[-214,0.0561],[-212,0.0561],[-210,0.0563],[-208,0.0589],[-206,0.061],[-204,0.0635],[-202,0.0645],[-200,0.0678],[-198,0.0654],[-196,0.0642],[-194,0.0654],[-192,0.068],[-190,0.0686],[-188,0.0665],[-184,0.0662],[-180,0.0638],[-178,0.0613],[-176,0.0604],[-174,0.0632],[-172,0.0704],[-170,0.08],[-168,0.0936],[-166,0.106],[-164,0.123],[-162,0.1388],[-160,0.164],[-158,0.1663],[-156,0.1871],[-154,0.205],[-152,0.2169],[-150,0.2257],[-148,0.2369],[-146,0.2442],[-144,0.2499],[-142,0.2562],[-140,0.2509],[-138,0.2529],[-136,0.2514],[-134,0.2487],[-132,0.2503],[-130,0.2555],[-128,0.264],[-124,0.257],[-122,0.2522],[-120,0.243],[-118,0.2218],[-116,0.1764],[-114,0.1651],[-112,0.139],[-110,0.1201],[-108,0.1083],[-106,0.1029],[-104,0.1006],[-102,0.0962],[-100,0.091],[-98,0.0906],[-96,0.0787],[-94,0.0794],[-92,0.0807],[-90,0.087],[-88,0.0873],[-86,0.0885],[-84,0.0951],[-82,0.1055],[-80,0.1223],[-78,0.1377],[-76,0.1415],[-74,0.1744],[-72,0.1911],[-70,0.2023],[-68,0.2192],[-66,0.2234],[-64,0.2135],[-62,0.2025],[-60,0.1367],[-58,0.1253],[-56,0.101],[-54,0.0912],[-52,0.0853],[-50,0.078],[-48,0.0756],[-46,0.0723],[-44,0.0666],[-42,0.065],[-40,0.0642],[-38,0.0693],[-36,0.0686],[-34,0.0677],[-32,0.0624],[-30,0.058],[-28,0.0507],[-26,0.0491],[-24,0.0482],[-22,0.0479],[-20,0.0503],[-16,0.0495],[-14,0.0471],[-12,0.0489],[-10,0.0465],[-8,0.0437],[-6,0.0419],[-4,0.0432],[-2,0.0419],[0,0.0397],[2,0.0391],[4,0.0392],[6,0.0407],[8,0.0398],[10,0.0405],[12,0.0387],[16,0.0379],[18,0.037],[22,0.0367],[24,0.0383],[26,0.039],[28,0.0391],[30,0.0383],[32,0.0381],[38,0.0411],[40,0.0397],[42,0.0386],[44,0.0369],[48,0.0362],[50,0.0366],[52,0.0378],[54,0.0388],[56,0.0389],[60,0.0383],[62,0.0398],[64,0.0405],[66,0.0414],[68,0.0401],[70,0.04],[74,0.0385],[76,0.0383],[80,0.0395],[82,0.0403],[84,0.0396],[86,0.0391],[88,0.0397],[90,0.0398],[94,0.0404],[96,0.0434],[98,0.048],[100,0.0521],[102,0.0588],[104,0.0638],[106,0.078],[108,0.0785],[110,0.0768],[112,0.074],[114,0.0699],[116,0.0676],[118,0.0644],[120,0.0637],[122,0.0642],[124,0.0602],[126,0.0601],[128,0.0574],[130,0.0574],[132,0.0568],[134,0.0592],[136,0.061],[138,0.0615],[140,0.0568],[144,0.0492],[146,0.051],[148,0.0535],[150,0.0556],[152,0.0549],[154,0.0521],[156,0.052],[158,0.0504],[160,0.0494],[162,0.0475],[164,0.0428],[166,0.0393],[170,0.0425],[172,0.043],[174,0.0424],[178,0.0423],[180,0.0409],[182,0.04],[184,0.0412],[188,0.0412],[192,0.0396],[194,0.0403],[196,0.0416],[198,0.0425],[200,0.0423],[202,0.0418],[204,0.0416],[206,0.042],[208,0.0413],[210,0.0413],[212,0.0406],[214,0.0386],[216,0.0382],[218,0.0382],[220,0.0391],[222,0.0413],[226,0.0438],[228,0.0428],[230,0.0425],[234,0.0467],[236,0.0486],[238,0.05],[240,0.0505],[242,0.0496],[244,0.0479],[246,0.0452],[248,0.0434],[250,0.0422],[254,0.0428],[256,0.042],[258,0.0446],[260,0.0447],[262,0.046],[264,0.0434],[266,0.0431],[268,0.0434],[270,0.0493],[272,0.0519],[274,0.0523],[276,0.0459],[278,0.0391],[280,0.0331],[284,0.0282],[286,0.0279],[288,0.0273],[290,0.0275],[292,0.0261],[294,0.0258],[296,0.0252],[298,0.0273],[300,0.0285],[302,0.0284],[306,0.027],[310,0.0277],[316,0.0282],[318,0.0289],[322,0.0276],[324,0.0278],[326,0.0282],[328,0.0278],[330,0.0275],[334,0.0288],[336,0.0377],[338,0.038],[342,0.0394],[344,0.0398],[346,0.0397],[348,0.0393],[350,0.0392],[352,0.0396],[354,0.0396],[356,0.0403],[360,0.0385],[362,0.038],[364,0.0363],[366,0.0333],[368,0.0322],[370,0.0333],[372,0.0336],[374,0.0326],[376,0.0325],[378,0.0316],[380,0.034],[382,0.0374],[384,0.0443],[386,0.0489],[388,0.0525],[390,0.0529],[392,0.0574],[394,0.0635],[396,0.0703],[398,0.0748],[400,0.0756],[402,0.0725],[404,0.0703],[406,0.064],[408,0.065],[410,0.0642],[412,0.0642],[416,0.0615],[418,0.06],[420,0.056],[422,0.0512],[424,0.0504],[426,0.0501],[428,0.0508],[430,0.0503],[432,0.0496],[434,0.0485],[436,0.0467],[438,0.0459],[440,0.0458],[442,0.0465],[444,0.0455],[446,0.046],[448,0.0461],[450,0.0479],[452,0.0483],[454,0.0479],[456,0.0502],[458,0.0533],[460,0.0577],[462,0.0601],[464,0.0594],[466,0.0614],[468,0.0638],[470,0.0685],[472,0.0748],[474,0.0761],[476,0.0727],[478,0.0663],[480,0.0669],[482,0.0698],[484,0.0702],[486,0.0735],[488,0.0688],[490,0.0653],[492,0.0663],[494,0.0745],[496,0.0791],[498,0.0862],[500,0.0842],[502,0.0827],[504,0.0728],[506,0.0669],[508,0.0617],[510,0.0597],[514,0.0509],[516,0.0554],[518,0.0559],[520,0.0599],[522,0.0612],[524,0.0594],[526,0.0522],[528,0.0512],[530,0.0499],[532,0.0496],[534,0.0483],[536,0.0477],[540,0.0476],[542,0.0507],[544,0.0529],[546,0.0547],[548,0.0556],[550,0.0557],[554,0.0547],[556,0.0562],[558,0.0582],[560,0.059],[562,0.0575],[564,0.0577],[566,0.0592],[568,0.0595],[570,0.0606],[574,0.0617],[576,0.0619],[580,0.0629],[582,0.0625],[584,0.0604],[586,0.0569],[588,0.0552],[592,0.0534],[594,0.0534],[596,0.0547],[598,0.0545],[600,0.0506],[602,0.0502],[604,0.0482],[606,0.048],[610,0.0465],[612,0.0446],[614,0.0432],[616,0.0422],[620,0.0422],[622,0.0387],[624,0.038],[626,0.0356],[628,0.0328],[630,0.0334],[632,0.0343],[636,0.0348],[638,0.0348],[640,0.0358],[642,0.0359],[644,0.0353],[648,0.0335],[654,0.0328],[656,0.0321],[658,0.0324],[662,0.0352],[664,0.035],[666,0.0339],[668,0.0337],[672,0.0342],[674,0.0342],[676,0.034],[678,0.0353],[680,0.0351],[682,0.0356],[684,0.0341],[686,0.0341],[688,0.0348],[690,0.0352],[692,0.0349],[694,0.0339],[696,0.0342],[698,0.0341],[700,0.0342],[702,0.0339],[704,0.034],[706,0.0335],[712,0.033],[718,0.0299],[720,0.0299]]
       }
       CTRL_L_mouth_lowerLipDepress {
        type=MovieSceneScriptingFloatChannel
        default=0
        thinned_from=492
        keys [[-262,0.0018],[-260,0.0011],[-258,0.0009],[-256,0.0003],[-254,0],[-252,0],[-250,0.0003],[-248,0.0004],[-242,0],[-240,0.0006],[-238,0.0002],[-226,0.0007],[-222,0.001],[-220,0.0004],[-218,0.0003],[-214,0.001],[-212,0.0011],[-210,0.0009],[-208,0.001],[-206,0.0006],[-202,0.0009],[-200,0.0003],[-196,0.001],[-190,0.0011],[-182,0.0005],[-180,0.001],[-176,0.0011],[-174,0.0019],[-172,0.0023],[-164,0.002],[-158,0.0022],[-156,0.0026],[-154,0.0048],[-152,0.0112],[-150,0.0393],[-148,0.0426],[-144,0.0583],[-142,0.0571],[-140,0.0638],[-138,0.0789],[-136,0.0729],[-134,0.0709],[-132,0.0782],[-130,0.085],[-128,0.0795],[-126,0.0745],[-124,0.0809],[-122,0.0811],[-120,0.0718],[-118,0.0495],[-116,0.0239],[-114,0.0204],[-112,0.013],[-110,0.0095],[-108,0.0075],[-106,0.0067],[-104,0.0063],[-100,0.004],[-98,0.0039],[-96,0.0027],[-88,0.0024],[-84,0.0018],[-80,0.0016],[-76,0.0018],[-74,0.0213],[-72,0.0321],[-68,0.0521],[-66,0.0516],[-64,0.0214],[-62,0.0183],[-60,0.0059],[-58,0.0053],[-56,0.0043],[-54,0.0042],[-50,0.0047],[-48,0.0048],[-46,0.004],[-44,0.0029],[-42,0.0029],[-38,0.0025],[-36,0.0028],[-34,0.0024],[-32,0.0026],[-28,0.0025],[-24,0.0028],[-16,0.0028],[-14,0.0026],[-12,0.0026],[-10,0.0025],[-8,0.0026],[-6,0.003],[-4,0.0026],[4,0.0028],[6,0.0026],[8,0.0028],[12,0.0025],[30,0.0027],[34,0.0025],[38,0.0025],[40,0.0028],[52,0.0026],[56,0.0029],[60,0.0025],[72,0.0024],[74,0.0028],[82,0.0029],[84,0.0028],[92,0.0027],[94,0.0025],[100,0.0033],[104,0.003],[108,0.0034],[110,0.0032],[112,0.0027],[114,0.0015],[116,0.0007],[118,0.0015],[120,0.0018],[122,0.0014],[124,0.0012],[126,0.0012],[128,0.0015],[130,0.0015],[132,0.0018],[134,0.0013],[136,0.0011],[138,0.0013],[144,0.0012],[146,0.0008],[150,0.0014],[152,0.0012],[154,0.0014],[160,0.0011],[162,0.0008],[166,0.0011],[168,0.001],[174,0.0014],[176,0.0008],[180,0.0007],[182,0.0005],[184,0.0009],[186,0.0008],[190,0.0003],[194,0.0005],[196,0.0004],[198,0.0009],[200,0.0011],[206,0.0004],[210,0.0005],[212,0.0003],[214,0.0004],[216,0.0001],[224,0.0005],[228,0.0003],[230,0.0007],[234,0.0002],[236,0.0007],[238,0.0004],[242,0.0007],[246,0.0004],[248,0.0006],[250,0.0003],[254,0.0004],[256,0.0001],[258,0.0004],[264,0],[266,0.0002],[268,0.001],[272,0.0009],[274,0.0016],[276,0.0019],[278,0.0006],[280,0.0013],[282,0.001],[284,0.0015],[286,0.0017],[292,0.0009],[294,0.0011],[298,0.0008],[302,0.0011],[304,0.001],[310,0.0014],[312,0.001],[316,0.001],[318,0.0013],[320,0.0009],[324,0.0015],[326,0.0013],[328,0.0009],[330,0.0011],[332,0.001],[334,0.0013],[336,0.0012],[344,0.0014],[348,0.0011],[356,0.0012],[358,0.0004],[360,0.0003],[362,0.0008],[364,0.0008],[368,0.0003],[376,0.0005],[380,0.0001],[382,0.0003],[384,0.0003],[388,0.0023],[390,0.0024],[392,0.002],[394,0.0011],[398,0.0019],[400,0.0019],[402,0.0016],[404,0.0024],[406,0.0019],[414,0.0017],[418,0.0019],[420,0.0018],[424,0.0018],[428,0.0015],[430,0.0017],[434,0.0016],[436,0.0017],[438,0.0015],[440,0.0019],[442,0.0018],[446,0.0019],[448,0.0018],[454,0.0022],[458,0.0016],[462,0.0017],[468,0.0015],[470,0.0005],[472,0.0017],[474,0.002],[476,0.0019],[478,0.0009],[480,0.0008],[482,0.001],[484,0.0011],[486,0.0004],[488,0.0001],[490,0.0005],[492,0.0002],[494,0.0002],[496,0],[498,0.0004],[500,0.0001],[502,0.0003],[504,0],[510,0.0002],[512,0],[518,0],[522,0.0014],[524,0.0014],[526,0.0023],[532,0.0027],[536,0.0023],[538,0.0018],[542,0.0017],[548,0.0004],[554,0],[562,0.0004],[564,0.0002],[570,0.0004],[572,0],[576,0.0001],[584,0],[586,0.0001],[588,0],[590,0.0004],[596,0.0007],[600,0.0012],[608,0.0013],[614,0.0021],[618,0.0015],[636,0.001],[642,0.0014],[644,0.001],[646,0.0008],[648,0.0014],[652,0.0015],[654,0.0013],[664,0.0017],[668,0.0015],[672,0.0017],[674,0.0014],[676,0.0016],[678,0.0014],[688,0.0015],[694,0.0012],[696,0.0014],[698,0.0011],[704,0.0012],[706,0.0014],[716,0.0015],[720,0.0013]]
       }
       CTRL_R_mouth_lowerLipDepress {
        type=MovieSceneScriptingFloatChannel
        default=0
        thinned_from=492
        keys [[-262,0.003],[-260,0.0017],[-258,0.0014],[-256,0.0007],[-254,0.0003],[-252,0.0002],[-248,0.0006],[-246,0.0003],[-242,0],[-240,0.0006],[-238,0.0002],[-236,0.0005],[-234,0.0006],[-232,0.0011],[-228,0.0011],[-226,0.001],[-222,0.0014],[-220,0.001],[-218,0.0009],[-216,0.0015],[-214,0.0018],[-208,0.0018],[-206,0.0013],[-202,0.0011],[-200,0.0005],[-196,0.0011],[-190,0.0011],[-188,0.0013],[-186,0.0013],[-182,0.0008],[-180,0.0014],[-178,0.0016],[-176,0.0016],[-174,0.0022],[-172,0.0025],[-158,0.003],[-156,0.0038],[-154,0.0107],[-152,0.0304],[-150,0.0802],[-148,0.0972],[-146,0.1104],[-144,0.1342],[-142,0.1374],[-140,0.1517],[-138,0.1766],[-136,0.1687],[-134,0.1683],[-132,0.1786],[-130,0.1894],[-128,0.1841],[-126,0.1766],[-124,0.1837],[-122,0.1774],[-120,0.1533],[-118,0.1196],[-116,0.0914],[-114,0.0841],[-112,0.0691],[-110,0.0611],[-108,0.0572],[-106,0.053],[-104,0.0493],[-102,0.0299],[-100,0.0202],[-98,0.0192],[-96,0.0069],[-94,0.0049],[-92,0.0047],[-90,0.0039],[-88,0.0037],[-86,0.0031],[-84,0.0028],[-76,0.0028],[-74,0.052],[-72,0.0814],[-70,0.1065],[-68,0.1236],[-66,0.1167],[-64,0.0615],[-62,0.0543],[-60,0.0248],[-58,0.0225],[-56,0.0193],[-54,0.0212],[-52,0.0204],[-50,0.0236],[-48,0.0251],[-46,0.0237],[-44,0.0179],[-42,0.009],[-40,0.0065],[-38,0.0044],[-34,0.0038],[-32,0.0039],[-30,0.0036],[-20,0.004],[-12,0.0041],[-8,0.0039],[-6,0.0041],[-4,0.004],[-2,0.0041],[2,0.0039],[4,0.0041],[10,0.0039],[12,0.0037],[16,0.0038],[20,0.0036],[30,0.0038],[34,0.0035],[40,0.0034],[42,0.0036],[44,0.0033],[54,0.0036],[58,0.0036],[60,0.0037],[66,0.0035],[68,0.0038],[70,0.0037],[72,0.0034],[74,0.0037],[78,0.0038],[84,0.0036],[92,0.0039],[94,0.0038],[102,0.0039],[104,0.0047],[106,0.0114],[108,0.0113],[110,0.0082],[112,0.005],[114,0.003],[116,0.0018],[124,0.0025],[130,0.0023],[132,0.0027],[134,0.0022],[136,0.0022],[138,0.0025],[142,0.0023],[144,0.0024],[146,0.0021],[148,0.0026],[150,0.004],[152,0.0036],[154,0.0065],[156,0.0066],[158,0.0073],[160,0.0073],[162,0.0038],[164,0.0046],[166,0.007],[168,0.0072],[170,0.0081],[172,0.0061],[174,0.007],[176,0.0039],[180,0.0031],[182,0.0029],[184,0.0067],[186,0.0064],[188,0.0051],[190,0.0028],[192,0.002],[196,0.0014],[198,0.002],[200,0.0022],[202,0.0018],[204,0.0017],[206,0.0013],[214,0.0014],[216,0.0011],[222,0.0016],[224,0.0015],[228,0.0009],[230,0.0016],[232,0.0017],[234,0.0012],[236,0.0014],[238,0.0011],[242,0.0015],[244,0.0019],[246,0.0019],[248,0.0013],[252,0.0057],[254,0.006],[256,0.005],[258,0.0073],[260,0.0099],[262,0.0111],[264,0.0084],[266,0.0099],[268,0.0061],[270,0.0033],[272,0.0021],[274,0.0027],[276,0.0038],[278,0.0014],[280,0.0018],[282,0.0013],[286,0.0019],[288,0.0018],[292,0.001],[294,0.0013],[296,0.0009],[298,0.0008],[302,0.0013],[304,0.0013],[306,0.0016],[308,0.0013],[310,0.0017],[312,0.001],[316,0.0012],[318,0.0019],[320,0.0013],[322,0.0015],[324,0.0021],[326,0.0019],[328,0.0013],[330,0.0015],[332,0.0015],[338,0.0021],[350,0.0019],[356,0.0021],[358,0.0011],[360,0.001],[362,0.0013],[364,0.0012],[368,0.0006],[370,0.001],[376,0.0012],[380,0.0009],[382,0.0011],[384,0.0008],[388,0.0034],[390,0.0033],[394,0.0017],[398,0.0029],[400,0.0029],[402,0.0023],[404,0.0032],[406,0.003],[412,0.0033],[414,0.0031],[422,0.0033],[426,0.0029],[432,0.0029],[434,0.0026],[440,0.0031],[444,0.003],[446,0.0034],[448,0.0031],[454,0.0033],[468,0.0025],[470,0.0013],[472,0.0024],[474,0.0028],[476,0.0028],[478,0.0016],[480,0.0013],[482,0.0018],[484,0.002],[488,0.0011],[490,0.0013],[492,0.0008],[494,0.0008],[496,0.0003],[498,0.0011],[500,0.0003],[502,0.0006],[504,0.0002],[506,0],[510,0.0002],[512,0],[516,0],[518,0.0002],[520,0.0016],[522,0.0023],[524,0.002],[526,0.0029],[530,0.0032],[534,0.0029],[536,0.003],[538,0.0026],[542,0.0023],[544,0.0019],[546,0.0012],[548,0.0007],[556,0.0003],[558,0.0004],[562,0.0007],[564,0.0006],[568,0.0008],[570,0.0006],[572,0],[576,0.0005],[582,0.0002],[584,0.0002],[586,0.0005],[588,0.0005],[592,0.0014],[596,0.0015],[600,0.0027],[614,0.0036],[620,0.0034],[622,0.003],[626,0.003],[630,0.0027],[632,0.003],[636,0.0026],[640,0.003],[642,0.003],[644,0.0025],[646,0.0023],[648,0.0029],[650,0.0028],[652,0.0029],[658,0.0027],[664,0.0031],[666,0.0028],[670,0.0029],[674,0.0027],[676,0.003],[678,0.0027],[682,0.0023],[686,0.0026],[694,0.0024],[696,0.0026],[700,0.0022],[702,0.0025],[704,0.0025],[706,0.0028],[708,0.0025],[720,0.0024]]
       }
       CTRL_R_mouth_stretch {
        type=MovieSceneScriptingFloatChannel
        default=0
        thinned_from=492
        keys [[-262,0.0515],[-260,0.0515],[-258,0.0492],[-256,0.0478],[-254,0.0485],[-252,0.0497],[-248,0.0485],[-246,0.0474],[-244,0.0471],[-242,0.0492],[-240,0.05],[-238,0.0512],[-236,0.0527],[-234,0.053],[-232,0.0551],[-230,0.0566],[-228,0.0565],[-226,0.0546],[-224,0.0536],[-220,0.0561],[-218,0.0546],[-216,0.0522],[-212,0.0523],[-210,0.0532],[-208,0.0552],[-206,0.0567],[-204,0.0588],[-202,0.0598],[-200,0.0633],[-198,0.0621],[-196,0.0616],[-194,0.0631],[-192,0.066],[-190,0.0667],[-188,0.0641],[-184,0.0639],[-182,0.0643],[-180,0.0642],[-178,0.0629],[-176,0.0638],[-174,0.0694],[-172,0.079],[-170,0.0911],[-168,0.1075],[-166,0.1223],[-164,0.1411],[-162,0.1581],[-160,0.1826],[-158,0.1846],[-156,0.2017],[-154,0.2119],[-152,0.2176],[-150,0.2199],[-148,0.2271],[-144,0.2369],[-142,0.2411],[-140,0.2351],[-138,0.238],[-136,0.238],[-134,0.2365],[-132,0.2384],[-130,0.244],[-128,0.2528],[-126,0.2488],[-124,0.2453],[-122,0.241],[-120,0.2328],[-118,0.2144],[-116,0.1786],[-114,0.1688],[-112,0.1446],[-110,0.1233],[-108,0.1094],[-106,0.1034],[-104,0.1008],[-100,0.0882],[-98,0.0877],[-96,0.0742],[-94,0.0749],[-92,0.0769],[-90,0.0866],[-88,0.0875],[-86,0.0907],[-84,0.0986],[-82,0.1106],[-80,0.1281],[-78,0.1425],[-76,0.1457],[-74,0.1723],[-72,0.1848],[-70,0.1938],[-68,0.2115],[-66,0.219],[-64,0.2118],[-62,0.2014],[-60,0.1396],[-58,0.128],[-56,0.1033],[-54,0.0926],[-52,0.0856],[-50,0.0777],[-48,0.0751],[-46,0.072],[-44,0.0648],[-42,0.0609],[-40,0.0605],[-38,0.0684],[-36,0.0721],[-34,0.071],[-32,0.0654],[-30,0.0596],[-28,0.0513],[-26,0.0487],[-24,0.0473],[-22,0.0467],[-20,0.0493],[-16,0.0498],[-14,0.0486],[-12,0.05],[-10,0.047],[-8,0.0431],[-6,0.0413],[-4,0.0433],[-2,0.0422],[0,0.0404],[2,0.0395],[4,0.0397],[6,0.0413],[8,0.0412],[10,0.0421],[12,0.0398],[16,0.0385],[18,0.037],[20,0.036],[22,0.0357],[24,0.0378],[26,0.0386],[28,0.0387],[30,0.0383],[32,0.0387],[34,0.0397],[36,0.0404],[38,0.0412],[42,0.0394],[44,0.0377],[46,0.0374],[48,0.0367],[50,0.0372],[54,0.0398],[56,0.04],[60,0.0398],[62,0.0411],[64,0.0414],[66,0.0423],[68,0.0408],[70,0.0407],[74,0.0392],[76,0.039],[78,0.0392],[80,0.0391],[82,0.0392],[84,0.0382],[86,0.0378],[88,0.0393],[90,0.0396],[92,0.0403],[94,0.0394],[96,0.041],[100,0.0503],[102,0.0577],[104,0.0629],[106,0.077],[108,0.0771],[110,0.075],[112,0.0721],[114,0.0684],[116,0.0664],[118,0.0634],[120,0.0627],[122,0.0634],[124,0.0598],[126,0.0596],[128,0.0557],[130,0.0557],[132,0.0551],[134,0.0581],[136,0.0616],[138,0.0627],[142,0.0539],[144,0.0502],[146,0.0522],[148,0.0547],[150,0.0569],[152,0.0561],[154,0.053],[156,0.0528],[158,0.0508],[160,0.0502],[162,0.0472],[164,0.0419],[166,0.0378],[168,0.0401],[170,0.0413],[172,0.041],[174,0.0398],[176,0.04],[178,0.0404],[180,0.0391],[182,0.0384],[184,0.0399],[188,0.0403],[192,0.0392],[194,0.04],[198,0.0421],[200,0.0423],[204,0.042],[206,0.0431],[210,0.0431],[212,0.0421],[214,0.0401],[216,0.0393],[218,0.0396],[220,0.0403],[222,0.0425],[226,0.0451],[228,0.0444],[230,0.0439],[232,0.0465],[234,0.0474],[236,0.0481],[240,0.0471],[242,0.046],[244,0.0444],[246,0.0415],[248,0.0398],[250,0.0393],[254,0.0408],[256,0.0395],[258,0.0422],[260,0.043],[262,0.0453],[264,0.0423],[268,0.041],[270,0.0461],[272,0.0464],[274,0.0435],[276,0.0349],[278,0.0287],[280,0.0255],[282,0.0248],[284,0.0244],[286,0.0233],[288,0.0215],[290,0.0219],[292,0.021],[294,0.0208],[296,0.0199],[298,0.0219],[300,0.0234],[302,0.0235],[304,0.0234],[306,0.0226],[308,0.0227],[312,0.022],[316,0.0222],[318,0.0237],[320,0.0237],[322,0.0227],[326,0.0226],[328,0.0222],[330,0.0216],[332,0.0227],[334,0.0229],[336,0.0344],[338,0.0346],[340,0.0353],[342,0.0364],[344,0.037],[346,0.0369],[348,0.0365],[350,0.0364],[352,0.0369],[354,0.037],[356,0.0379],[358,0.0378],[360,0.0373],[362,0.0386],[364,0.0377],[366,0.0348],[368,0.0337],[370,0.0347],[372,0.0349],[374,0.0341],[376,0.0339],[378,0.0332],[382,0.0381],[384,0.0442],[386,0.0479],[388,0.0514],[390,0.0525],[392,0.0578],[394,0.0637],[396,0.0688],[398,0.0717],[400,0.0722],[402,0.0697],[404,0.0684],[406,0.0622],[408,0.0617],[410,0.0594],[412,0.0584],[414,0.0567],[416,0.0553],[418,0.0542],[420,0.0507],[422,0.0466],[424,0.0459],[426,0.0462],[428,0.047],[430,0.0471],[432,0.0466],[434,0.0468],[436,0.0454],[438,0.0446],[440,0.0436],[442,0.0439],[444,0.0431],[446,0.0434],[448,0.0432],[450,0.0445],[452,0.0447],[454,0.0442],[456,0.0473],[458,0.0509],[460,0.0555],[462,0.0573],[464,0.0563],[466,0.0575],[468,0.059],[470,0.062],[472,0.0682],[474,0.0707],[476,0.0688],[478,0.0641],[482,0.0658],[484,0.0657],[486,0.066],[488,0.0612],[490,0.0595],[492,0.063],[494,0.0725],[496,0.0787],[498,0.0877],[500,0.087],[502,0.086],[504,0.0768],[508,0.0675],[510,0.0654],[512,0.0609],[514,0.0562],[516,0.0596],[518,0.06],[520,0.0632],[522,0.0631],[524,0.0607],[526,0.052],[528,0.051],[530,0.0505],[532,0.0512],[534,0.0499],[536,0.0489],[540,0.0491],[542,0.0522],[544,0.0537],[548,0.0555],[550,0.0558],[552,0.0553],[554,0.0549],[556,0.0565],[558,0.0578],[560,0.0578],[562,0.0565],[564,0.0573],[566,0.059],[568,0.0593],[572,0.0614],[574,0.062],[576,0.0624],[580,0.0637],[582,0.0638],[584,0.0619],[586,0.0576],[588,0.055],[590,0.0545],[592,0.0558],[594,0.0561],[596,0.0599],[598,0.0603],[600,0.0541],[602,0.0534],[604,0.0491],[606,0.0487],[610,0.0466],[612,0.0447],[616,0.0428],[620,0.0434],[622,0.0406],[624,0.0399],[628,0.0339],[630,0.0341],[632,0.0352],[634,0.0352],[636,0.0361],[638,0.0361],[640,0.0371],[642,0.0372],[644,0.036],[646,0.0352],[648,0.0335],[654,0.0328],[656,0.0314],[658,0.0317],[660,0.0329],[662,0.0343],[664,0.0337],[666,0.0323],[668,0.0316],[670,0.0316],[676,0.0326],[678,0.0337],[680,0.0335],[682,0.0338],[684,0.033],[686,0.0333],[688,0.034],[690,0.0343],[692,0.034],[694,0.0331],[696,0.0339],[698,0.0341],[700,0.0346],[702,0.0345],[704,0.035],[708,0.0347],[712,0.0343],[716,0.0318],[718,0.0312],[720,0.0312]]
       }
       CTRL_L_mouth_towardsD {
        type=MovieSceneScriptingFloatChannel
        default=0
        thinned_from=492
        keys [[-262,0.0034],[-260,0.0023],[-258,0.0022],[-254,0.0013],[-252,0.0011],[-248,0.0017],[-246,0.0012],[-242,0.001],[-240,0.0017],[-238,0.0016],[-234,0.0019],[-230,0.0023],[-226,0.002],[-222,0.0021],[-220,0.0015],[-218,0.0014],[-214,0.0027],[-212,0.0027],[-208,0.0026],[-206,0.0022],[-202,0.0024],[-200,0.0018],[-196,0.0023],[-186,0.0022],[-184,0.0018],[-182,0.0017],[-176,0.0025],[-172,0.0034],[-170,0.0036],[-166,0.0035],[-160,0.0038],[-158,0.0038],[-154,0.0034],[-152,0.0037],[-148,0.0035],[-146,0.0038],[-142,0.0037],[-136,0.004],[-130,0.0037],[-128,0.0038],[-122,0.0037],[-112,0.0042],[-108,0.0038],[-104,0.0042],[-102,0.004],[-98,0.0041],[-94,0.0038],[-88,0.0036],[-84,0.0031],[-76,0.0035],[-74,0.0039],[-70,0.0041],[-66,0.0039],[-64,0.0042],[-62,0.0042],[-60,0.0038],[-56,0.0038],[-54,0.0039],[-48,0.0035],[-46,0.0032],[-44,0.0032],[-42,0.0034],[-40,0.004],[-38,0.0039],[-34,0.0041],[-30,0.0039],[-28,0.0041],[-24,0.0039],[-16,0.0045],[-14,0.0043],[-12,0.0043],[-10,0.004],[-8,0.004],[-6,0.0041],[-4,0.004],[-2,0.0041],[2,0.0039],[4,0.0041],[6,0.004],[14,0.0044],[26,0.0039],[32,0.0041],[38,0.0041],[40,0.0043],[42,0.0042],[44,0.0038],[56,0.0037],[60,0.0042],[66,0.0038],[84,0.0039],[86,0.0042],[90,0.0042],[92,0.004],[96,0.004],[98,0.0038],[104,0.0041],[106,0.0038],[110,0.004],[112,0.0037],[116,0.0019],[124,0.0028],[126,0.0027],[128,0.0025],[130,0.0025],[132,0.0031],[134,0.0027],[136,0.0027],[138,0.003],[140,0.0026],[144,0.0027],[146,0.0023],[148,0.0025],[150,0.003],[152,0.0027],[156,0.0028],[160,0.0023],[162,0.0024],[164,0.0021],[166,0.0023],[168,0.0018],[174,0.0021],[176,0.0017],[178,0.0018],[180,0.0016],[186,0.0017],[190,0.0015],[194,0.0018],[196,0.0017],[200,0.0023],[204,0.0022],[206,0.0016],[212,0.0016],[214,0.0018],[216,0.0016],[226,0.0021],[228,0.0017],[230,0.0023],[234,0.0018],[238,0.0018],[242,0.0021],[246,0.0018],[248,0.0019],[250,0.0014],[254,0.0018],[256,0.0016],[260,0.0019],[262,0.0015],[264,0.0009],[268,0.0023],[272,0.0021],[276,0.0036],[278,0.0022],[280,0.0033],[282,0.003],[286,0.0034],[288,0.0033],[292,0.0024],[294,0.0027],[296,0.0023],[298,0.0023],[302,0.0028],[304,0.0027],[306,0.0029],[308,0.0028],[310,0.003],[312,0.0027],[316,0.0025],[318,0.0031],[320,0.0027],[324,0.0032],[326,0.0029],[328,0.0028],[330,0.003],[332,0.0028],[334,0.0029],[336,0.0026],[338,0.0034],[342,0.0039],[344,0.0036],[346,0.0027],[348,0.0021],[350,0.0024],[356,0.0028],[358,0.002],[364,0.002],[368,0.0015],[374,0.0021],[378,0.002],[380,0.0016],[382,0.0017],[384,0.0013],[388,0.0037],[390,0.0038],[392,0.0033],[394,0.0026],[398,0.0038],[400,0.0038],[402,0.0034],[404,0.0038],[408,0.0035],[412,0.0036],[414,0.0035],[416,0.0036],[420,0.0034],[422,0.0036],[428,0.0032],[432,0.0034],[434,0.0031],[436,0.0033],[438,0.003],[440,0.0036],[444,0.0032],[446,0.0033],[448,0.0032],[454,0.0034],[462,0.0029],[466,0.0033],[468,0.0031],[470,0.0021],[472,0.0033],[474,0.0035],[476,0.0034],[478,0.0023],[480,0.0022],[484,0.0024],[488,0.0016],[490,0.0018],[492,0.0015],[494,0.0017],[496,0.0011],[498,0.0017],[500,0.001],[502,0.0019],[504,0.0016],[506,0.0015],[508,0.002],[510,0.0021],[512,0.0015],[514,0.0007],[516,0.0002],[518,0.0004],[520,0.0022],[522,0.0028],[524,0.0025],[526,0.0036],[530,0.0038],[536,0.0036],[538,0.0032],[540,0.0031],[542,0.0033],[544,0.0031],[548,0.0018],[556,0.0016],[562,0.002],[564,0.0019],[570,0.0021],[572,0.001],[576,0.0017],[584,0.0015],[588,0.0016],[590,0.0021],[596,0.0022],[600,0.003],[608,0.0028],[614,0.0033],[620,0.0031],[622,0.0029],[628,0.0029],[630,0.0028],[632,0.003],[636,0.0027],[642,0.0031],[644,0.0026],[646,0.0023],[648,0.0027],[652,0.0029],[654,0.0028],[656,0.003],[660,0.003],[662,0.0031],[666,0.0028],[670,0.0029],[674,0.0027],[676,0.003],[682,0.0026],[694,0.0029],[700,0.0026],[702,0.0028],[706,0.0028],[708,0.0026],[720,0.0027]]
       }
       CTRL_R_mouth_funnelU {
        type=MovieSceneScriptingFloatChannel
        default=0
        thinned_from=492
        keys [[-262,0.0025],[-260,0.0014],[-258,0.0012],[-254,0.0003],[-252,0.0002],[-248,0.0005],[-242,0],[-240,0.0004],[-238,0.0002],[-236,0.0005],[-234,0.0006],[-228,0.0015],[-224,0.0014],[-222,0.0014],[-220,0.0008],[-218,0.0007],[-214,0.0019],[-212,0.002],[-208,0.0016],[-206,0.0011],[-204,0.0013],[-202,0.0012],[-200,0.0005],[-196,0.001],[-182,0.0008],[-180,0.0013],[-176,0.0016],[-172,0.0024],[-166,0.0028],[-162,0.0025],[-158,0.0026],[-154,0.0025],[-152,0.0026],[-148,0.0021],[-144,0.0024],[-142,0.0023],[-140,0.0026],[-134,0.0027],[-126,0.0022],[-120,0.0026],[-114,0.0025],[-110,0.0027],[-108,0.0026],[-104,0.003],[-80,0.0026],[-76,0.0028],[-72,0.0024],[-66,0.0025],[-64,0.0028],[-56,0.0024],[-46,0.0022],[-44,0.0025],[-42,0.0025],[-40,0.0028],[-38,0.0027],[-36,0.0032],[-32,0.0027],[-26,0.0029],[-24,0.0028],[-22,0.0029],[-20,0.0028],[-14,0.003],[-12,0.0033],[0,0.0029],[2,0.0031],[6,0.003],[8,0.0031],[10,0.003],[12,0.0033],[18,0.0031],[36,0.0032],[38,0.003],[42,0.0032],[44,0.0029],[50,0.0028],[60,0.0032],[64,0.003],[70,0.0031],[72,0.0029],[74,0.0031],[90,0.0031],[92,0.0033],[96,0.003],[102,0.0026],[104,0.0028],[106,0.0026],[108,0.0028],[112,0.0025],[114,0.0017],[116,0.0007],[118,0.0011],[124,0.0015],[130,0.0016],[132,0.002],[136,0.0016],[138,0.002],[144,0.0017],[146,0.0015],[148,0.0016],[150,0.0019],[152,0.0015],[156,0.0016],[160,0.0013],[162,0.0014],[164,0.0011],[166,0.0012],[168,0.0009],[174,0.0016],[176,0.0009],[178,0.0011],[188,0.0008],[194,0.0009],[196,0.0008],[200,0.0016],[204,0.0012],[206,0.0007],[210,0.0009],[216,0.0008],[218,0.0009],[220,0.0008],[222,0.0011],[226,0.001],[228,0.0005],[232,0.0013],[234,0.0006],[236,0.0009],[238,0.0005],[244,0.0007],[246,0.0004],[248,0.0005],[250,0.0002],[254,0.0002],[256,0],[258,0.0004],[262,0],[264,0],[266,0.0003],[268,0.0012],[272,0.0011],[274,0.0023],[276,0.0099],[278,0.0095],[280,0.0121],[282,0.0122],[284,0.0133],[286,0.0128],[288,0.0117],[290,0.0091],[292,0.0099],[294,0.0097],[296,0.011],[298,0.0074],[302,0.011],[304,0.0107],[306,0.0126],[308,0.0128],[310,0.0097],[312,0.0095],[314,0.0099],[316,0.0089],[318,0.0095],[320,0.0091],[322,0.0096],[326,0.0089],[328,0.007],[330,0.0069],[336,0.0022],[340,0.0023],[342,0.0026],[344,0.0022],[346,0.0014],[348,0.001],[354,0.0015],[356,0.0015],[358,0.0007],[360,0.0007],[362,0.0011],[364,0.0011],[368,0.0006],[370,0.0009],[372,0.0008],[374,0.0012],[376,0.0013],[380,0.0007],[384,0.0007],[386,0.002],[388,0.0029],[390,0.0028],[394,0.0016],[396,0.0022],[398,0.0026],[400,0.0025],[402,0.0022],[404,0.003],[414,0.0025],[422,0.0028],[428,0.0021],[432,0.0024],[434,0.0021],[436,0.0023],[438,0.0022],[440,0.0025],[444,0.0025],[446,0.0028],[448,0.0024],[454,0.0026],[458,0.0022],[460,0.0024],[462,0.0021],[464,0.002],[466,0.0024],[468,0.0021],[470,0.0012],[472,0.0025],[474,0.0029],[476,0.0028],[478,0.0014],[480,0.0014],[482,0.0018],[484,0.0018],[486,0.0011],[488,0.0008],[490,0.0012],[492,0.0008],[494,0.0008],[496,0.0003],[498,0.0008],[500,0.0002],[502,0.0007],[504,0.0002],[506,0],[508,0.0008],[514,0],[518,0.0001],[520,0.0015],[522,0.0022],[524,0.0018],[526,0.0028],[530,0.0032],[536,0.0031],[538,0.0026],[542,0.0023],[544,0.0019],[548,0.0005],[552,0.0003],[556,0.0004],[558,0.0003],[562,0.0006],[564,0.0003],[566,0.0005],[570,0.0006],[572,0.0001],[578,0.0005],[582,0.0002],[584,0.0002],[586,0.0006],[592,0.0013],[596,0.0014],[600,0.0023],[610,0.0022],[614,0.0026],[626,0.0023],[636,0.0018],[642,0.0022],[644,0.0017],[646,0.0016],[648,0.002],[650,0.0021],[654,0.002],[664,0.0024],[666,0.0021],[670,0.0022],[674,0.002],[676,0.0021],[684,0.0017],[686,0.002],[692,0.0018],[694,0.0021],[696,0.0018],[698,0.0018],[700,0.0016],[702,0.002],[704,0.002],[706,0.0022],[714,0.002],[716,0.0021],[720,0.002]]
       }
       CTRL_L_mouth_towardsU {
        type=MovieSceneScriptingFloatChannel
        default=0
        thinned_from=492
        keys [[-262,0.0092],[-260,0.0082],[-258,0.008],[-256,0.0071],[-252,0.0068],[-248,0.0078],[-246,0.0068],[-244,0.0065],[-242,0.0066],[-240,0.0073],[-234,0.0069],[-232,0.0066],[-228,0.0071],[-222,0.0068],[-220,0.0062],[-218,0.006],[-214,0.0065],[-212,0.0065],[-210,0.0063],[-208,0.0066],[-206,0.0062],[-202,0.0069],[-200,0.0058],[-198,0.0064],[-192,0.0074],[-188,0.0076],[-186,0.0075],[-184,0.0064],[-182,0.0064],[-176,0.0076],[-174,0.0083],[-172,0.0081],[-168,0.0066],[-166,0.0063],[-164,0.0064],[-158,0.0065],[-156,0.0061],[-154,0.0053],[-152,0.0055],[-150,0.0049],[-148,0.0046],[-146,0.0051],[-144,0.0053],[-142,0.0049],[-140,0.005],[-138,0.0049],[-134,0.0051],[-130,0.0046],[-124,0.0052],[-122,0.0052],[-120,0.0057],[-118,0.0071],[-116,0.0093],[-114,0.0098],[-112,0.0109],[-106,0.0117],[-104,0.0118],[-102,0.0125],[-96,0.0124],[-94,0.0118],[-92,0.0115],[-90,0.0097],[-88,0.0091],[-86,0.0072],[-84,0.0062],[-82,0.0057],[-80,0.0056],[-76,0.0062],[-74,0.0059],[-70,0.006],[-68,0.0058],[-66,0.0059],[-64,0.008],[-62,0.0084],[-60,0.0099],[-56,0.0103],[-52,0.0104],[-48,0.0107],[-46,0.0143],[-44,0.0167],[-42,0.0162],[-40,0.0151],[-36,0.0103],[-34,0.0103],[-30,0.0095],[-28,0.01],[-24,0.0095],[-22,0.0101],[-20,0.0098],[-18,0.0104],[-16,0.0103],[-14,0.0095],[-8,0.0103],[-4,0.0101],[-2,0.0102],[0,0.0098],[6,0.0101],[8,0.01],[10,0.0091],[14,0.0096],[20,0.01],[26,0.0088],[28,0.009],[30,0.0095],[32,0.0088],[34,0.0084],[36,0.0082],[40,0.0085],[42,0.0084],[44,0.0078],[48,0.0081],[50,0.0084],[56,0.0087],[60,0.0083],[64,0.0086],[66,0.0078],[68,0.0074],[70,0.0075],[72,0.0081],[76,0.0078],[78,0.0079],[84,0.0091],[88,0.0087],[90,0.0088],[92,0.0115],[94,0.0107],[96,0.0119],[98,0.0121],[100,0.0114],[104,0.0118],[108,0.0127],[110,0.0127],[114,0.0106],[116,0.0093],[118,0.0087],[120,0.0086],[122,0.0094],[126,0.0095],[128,0.01],[132,0.0098],[136,0.009],[138,0.0092],[140,0.0088],[144,0.009],[146,0.008],[148,0.0083],[150,0.0088],[152,0.0084],[154,0.0091],[156,0.0091],[158,0.0087],[160,0.0077],[162,0.0083],[164,0.0081],[168,0.0075],[172,0.0082],[174,0.0077],[176,0.0063],[180,0.0068],[182,0.0068],[184,0.0071],[192,0.0062],[196,0.0052],[198,0.0057],[200,0.0067],[204,0.0063],[206,0.0056],[208,0.006],[212,0.0062],[214,0.0066],[216,0.0065],[218,0.0072],[220,0.0069],[222,0.0078],[224,0.008],[226,0.0077],[228,0.0076],[230,0.0087],[232,0.0091],[236,0.0106],[238,0.0106],[240,0.0117],[242,0.0122],[244,0.0121],[246,0.0115],[248,0.0116],[250,0.0113],[252,0.0117],[254,0.0119],[256,0.0113],[258,0.0112],[260,0.012],[262,0.011],[264,0.0111],[266,0.0128],[268,0.0143],[270,0.015],[272,0.0172],[274,0.017],[276,0.0156],[278,0.0105],[282,0.0107],[284,0.0112],[286,0.0125],[288,0.0123],[290,0.0116],[292,0.0112],[294,0.0121],[296,0.0109],[298,0.0104],[300,0.0115],[302,0.012],[304,0.0116],[308,0.0117],[310,0.0122],[312,0.0113],[314,0.011],[316,0.0099],[318,0.0106],[320,0.0095],[324,0.0095],[328,0.0088],[330,0.0092],[332,0.0089],[336,0.0087],[338,0.01],[340,0.0095],[342,0.0093],[344,0.0095],[346,0.0092],[348,0.0079],[354,0.0085],[356,0.0085],[358,0.0069],[360,0.0063],[364,0.0061],[366,0.0053],[368,0.0055],[370,0.0063],[372,0.0063],[374,0.0076],[376,0.0078],[380,0.0075],[382,0.0083],[384,0.0102],[386,0.0147],[388,0.0162],[390,0.0142],[392,0.0088],[394,0.0054],[396,0.0056],[398,0.0063],[400,0.0064],[402,0.0061],[404,0.008],[406,0.0069],[408,0.0065],[410,0.0067],[414,0.0068],[418,0.007],[420,0.0066],[422,0.0071],[426,0.0067],[434,0.0066],[436,0.0069],[438,0.0069],[440,0.0072],[442,0.0071],[444,0.0066],[446,0.0068],[448,0.0065],[454,0.0068],[456,0.0063],[460,0.006],[462,0.0055],[466,0.0057],[468,0.0054],[470,0.0037],[472,0.0056],[474,0.0067],[476,0.0069],[478,0.004],[480,0.0037],[482,0.0046],[484,0.0048],[488,0.0036],[490,0.0039],[492,0.0035],[494,0.0045],[496,0.0054],[498,0.0059],[500,0.005],[502,0.0055],[506,0.0039],[510,0.0058],[512,0.0051],[514,0.0048],[516,0.0057],[518,0.006],[520,0.0089],[522,0.0097],[524,0.0091],[530,0.0088],[532,0.009],[536,0.0108],[538,0.0096],[540,0.0093],[542,0.0081],[544,0.0071],[548,0.0057],[550,0.0058],[554,0.0052],[558,0.005],[562,0.0054],[564,0.0053],[566,0.0056],[570,0.0056],[572,0.0039],[576,0.0044],[582,0.0042],[584,0.0049],[586,0.006],[590,0.0062],[592,0.0066],[594,0.0066],[596,0.0057],[598,0.0061],[600,0.0073],[602,0.0073],[606,0.0079],[608,0.009],[610,0.0096],[614,0.0097],[620,0.0094],[622,0.0083],[624,0.0084],[626,0.0093],[632,0.0097],[636,0.0092],[638,0.0094],[640,0.0093],[642,0.0094],[644,0.0093],[646,0.009],[648,0.0099],[650,0.01],[652,0.0098],[654,0.0095],[658,0.0097],[660,0.0093],[662,0.0094],[664,0.01],[666,0.0096],[668,0.01],[670,0.0099],[674,0.0101],[678,0.0099],[680,0.0094],[682,0.0088],[684,0.0088],[686,0.0091],[690,0.0089],[692,0.009],[696,0.0087],[698,0.0088],[700,0.0084],[702,0.009],[706,0.0089],[708,0.0085],[710,0.0083],[716,0.0085],[720,0.0083]]
       }
       CTRL_R_mouth_lipBiteD {
        type=MovieSceneScriptingFloatChannel
        default=0
        thinned_from=492
        keys [[-262,0.0022],[-260,0.0016],[-258,0.0008],[-254,0.0001],[-252,0.0001],[-250,0.0003],[-248,0.0008],[-246,0.0005],[-244,0.0004],[-242,0],[-240,0.0008],[-238,0.0005],[-234,0.0007],[-232,0.001],[-228,0.0011],[-226,0.0009],[-222,0.0011],[-218,0.0004],[-214,0.0015],[-202,0.0011],[-200,0.0007],[-198,0.0008],[-194,0.0013],[-190,0.0012],[-186,0.0013],[-184,0.001],[-182,0.0009],[-178,0.0016],[-176,0.0017],[-170,0.0027],[-168,0.0029],[-166,0.0037],[-164,0.0048],[-162,0.008],[-160,0.028],[-158,0.033],[-156,0.0785],[-154,0.1485],[-152,0.1975],[-150,0.2681],[-148,0.3284],[-144,0.3752],[-142,0.384],[-140,0.3916],[-138,0.4044],[-134,0.4024],[-132,0.3976],[-130,0.3888],[-128,0.3849],[-124,0.3883],[-122,0.3786],[-120,0.3591],[-118,0.3183],[-116,0.243],[-114,0.2296],[-112,0.1956],[-110,0.1689],[-108,0.1522],[-106,0.1409],[-104,0.1358],[-102,0.1065],[-100,0.0898],[-98,0.0883],[-96,0.0639],[-92,0.0646],[-90,0.0603],[-88,0.0586],[-86,0.0531],[-84,0.0538],[-82,0.0605],[-80,0.0733],[-78,0.1005],[-76,0.111],[-74,0.2451],[-72,0.309],[-70,0.3391],[-68,0.3409],[-66,0.3316],[-64,0.283],[-62,0.2599],[-60,0.1408],[-58,0.1332],[-56,0.1176],[-54,0.1119],[-52,0.1105],[-50,0.1119],[-48,0.1127],[-46,0.1168],[-44,0.1029],[-42,0.0857],[-40,0.0834],[-38,0.0813],[-36,0.0721],[-34,0.0637],[-32,0.0537],[-30,0.0456],[-28,0.0416],[-26,0.0413],[-24,0.0392],[-22,0.0376],[-20,0.0385],[-18,0.0365],[-16,0.0329],[-14,0.0336],[-12,0.0356],[-10,0.0365],[-8,0.0355],[-6,0.0342],[-4,0.0353],[-2,0.0327],[0,0.0297],[2,0.0302],[6,0.0306],[8,0.0301],[10,0.0279],[12,0.0237],[16,0.0231],[20,0.0205],[22,0.0207],[26,0.0242],[30,0.0231],[32,0.0238],[34,0.0242],[38,0.0245],[40,0.0254],[42,0.0258],[48,0.0257],[50,0.027],[52,0.0287],[54,0.0301],[56,0.0304],[58,0.03],[60,0.0293],[62,0.0289],[64,0.0292],[66,0.0301],[68,0.0287],[72,0.0288],[74,0.0312],[76,0.0305],[78,0.0296],[80,0.0302],[82,0.0292],[84,0.0285],[86,0.031],[88,0.0331],[90,0.0333],[92,0.03],[94,0.0316],[96,0.0387],[100,0.0437],[102,0.0439],[104,0.0451],[106,0.049],[108,0.0462],[110,0.0385],[112,0.0338],[114,0.0318],[118,0.023],[120,0.0223],[124,0.0227],[126,0.0224],[128,0.0176],[130,0.0176],[132,0.0186],[134,0.0209],[136,0.0208],[146,0.0217],[148,0.0226],[150,0.025],[152,0.0262],[154,0.0256],[156,0.0258],[158,0.0277],[160,0.0283],[162,0.0277],[164,0.0283],[166,0.0304],[168,0.0316],[170,0.0308],[172,0.0285],[174,0.0281],[176,0.0284],[178,0.0282],[180,0.0286],[182,0.0287],[184,0.0306],[186,0.0306],[188,0.0295],[190,0.026],[192,0.0255],[194,0.0258],[196,0.025],[198,0.0248],[200,0.0238],[202,0.0231],[206,0.023],[208,0.0226],[210,0.0226],[212,0.0211],[214,0.0214],[216,0.0214],[218,0.0204],[220,0.0198],[224,0.0169],[226,0.0152],[228,0.0138],[230,0.0135],[232,0.0151],[234,0.0147],[236,0.0153],[238,0.0172],[240,0.0186],[242,0.0183],[244,0.0166],[246,0.0166],[248,0.0164],[250,0.0166],[252,0.0181],[254,0.0184],[256,0.0174],[258,0.0169],[260,0.0178],[262,0.0184],[264,0.0171],[266,0.0165],[268,0.0162],[270,0.0156],[272,0.0146],[274,0.0143],[276,0.0142],[278,0.0135],[280,0.0172],[282,0.019],[284,0.0195],[288,0.0165],[290,0.0166],[294,0.0155],[296,0.0166],[298,0.0158],[300,0.0142],[304,0.0164],[306,0.018],[308,0.0184],[310,0.0175],[312,0.017],[314,0.0159],[316,0.0157],[318,0.0174],[320,0.0179],[324,0.0161],[326,0.0149],[328,0.0141],[332,0.0151],[334,0.0144],[336,0.0126],[340,0.0098],[342,0.0087],[344,0.009],[346,0.0091],[348,0.0082],[350,0.0069],[352,0.0076],[356,0.0079],[358,0.0062],[360,0.005],[364,0.0046],[368,0.0033],[374,0.0017],[378,0.0014],[380,0.0009],[382,0.001],[384,0.0009],[386,0.0017],[388,0.0039],[390,0.004],[392,0.0032],[394,0.0021],[396,0.0023],[398,0.0036],[400,0.004],[402,0.0052],[404,0.0103],[406,0.0111],[408,0.007],[410,0.0069],[416,0.0056],[418,0.0057],[422,0.0046],[426,0.0043],[430,0.0043],[432,0.0047],[436,0.0038],[438,0.004],[440,0.0049],[444,0.0059],[446,0.0068],[448,0.0073],[450,0.0065],[452,0.0065],[454,0.0068],[456,0.0062],[458,0.0062],[462,0.0072],[464,0.0073],[466,0.0083],[468,0.0079],[470,0.0043],[472,0.0027],[476,0.0024],[478,0.0015],[480,0.0012],[482,0.0016],[484,0.0017],[488,0.0009],[490,0.0011],[498,0.0004],[500,0.0004],[504,0.0008],[506,0.0004],[510,0.0011],[512,0.0007],[514,0],[518,0],[522,0.0021],[532,0.0028],[542,0.0023],[544,0.0021],[548,0.0009],[552,0.0006],[556,0.0007],[558,0.0005],[560,0.0005],[562,0.0008],[568,0.0006],[570,0.0007],[572,0],[576,0.0006],[578,0.0007],[582,0.0003],[588,0.0005],[592,0.0012],[596,0.0014],[600,0.002],[604,0.0018],[606,0.0023],[610,0.0059],[612,0.0071],[614,0.0076],[620,0.0077],[622,0.0083],[624,0.0084],[626,0.0089],[628,0.0088],[630,0.0093],[632,0.0092],[636,0.0085],[638,0.0084],[640,0.0085],[642,0.0082],[644,0.0075],[646,0.0071],[650,0.0084],[654,0.0093],[658,0.0094],[662,0.0088],[664,0.0083],[668,0.0081],[672,0.0087],[674,0.0084],[676,0.0079],[680,0.0082],[682,0.0075],[684,0.0073],[686,0.0076],[688,0.0075],[690,0.0077],[692,0.0077],[694,0.0074],[696,0.0076],[698,0.0074],[702,0.0065],[704,0.0068],[706,0.0068],[710,0.0058],[712,0.0057],[716,0.0063],[718,0.0057],[720,0.0057]]
       }
       CTRL_R_mouth_funnelD {
        type=MovieSceneScriptingFloatChannel
        default=0
        thinned_from=492
        keys [[-262,0.0145],[-260,0.0182],[-258,0.017],[-256,0.0174],[-254,0.0194],[-252,0.02],[-250,0.0212],[-248,0.0238],[-246,0.0269],[-244,0.0276],[-242,0.0259],[-240,0.0253],[-238,0.024],[-236,0.0233],[-234,0.0228],[-232,0.02],[-230,0.0207],[-226,0.0189],[-224,0.0193],[-222,0.0188],[-220,0.0165],[-218,0.0149],[-216,0.015],[-214,0.0138],[-212,0.0136],[-210,0.0131],[-208,0.0143],[-206,0.0123],[-204,0.0126],[-202,0.0103],[-200,0.0086],[-198,0.009],[-194,0.0081],[-192,0.0075],[-190,0.0074],[-188,0.01],[-186,0.0101],[-184,0.0111],[-182,0.0102],[-180,0.0094],[-178,0.0092],[-176,0.0087],[-174,0.007],[-172,0.0051],[-170,0.0043],[-166,0.0036],[-164,0.0035],[-162,0.0036],[-160,0.0083],[-158,0.0092],[-154,0.0267],[-152,0.0338],[-150,0.0291],[-148,0.0202],[-146,0.0159],[-144,0.0132],[-140,0.0112],[-136,0.0101],[-134,0.0081],[-132,0.01],[-130,0.0096],[-128,0.0088],[-126,0.007],[-124,0.0056],[-120,0.004],[-116,0.0035],[-112,0.0032],[-110,0.0028],[-98,0.0027],[-86,0.0032],[-82,0.0031],[-76,0.0035],[-74,0.0159],[-72,0.0138],[-70,0.0095],[-68,0.006],[-66,0.0043],[-64,0.0039],[-62,0.0037],[-60,0.0031],[-56,0.0027],[-54,0.0028],[-52,0.0024],[-46,0.0021],[-44,0.0026],[-36,0.0031],[-34,0.003],[-32,0.0033],[-28,0.0032],[-26,0.0033],[-24,0.0031],[-22,0.0033],[-18,0.003],[-14,0.0033],[-8,0.003],[-6,0.0031],[-2,0.003],[0,0.0031],[2,0.0029],[8,0.0029],[10,0.0028],[12,0.0031],[38,0.0029],[40,0.0031],[64,0.0029],[66,0.0028],[78,0.0028],[80,0.003],[84,0.0029],[102,0.003],[104,0.0031],[106,0.0027],[110,0.0028],[112,0.0027],[116,0.0008],[118,0.0012],[124,0.0019],[130,0.0017],[132,0.0022],[136,0.0016],[138,0.002],[142,0.0018],[144,0.0019],[146,0.0016],[148,0.0018],[150,0.0023],[152,0.0019],[156,0.0019],[160,0.0014],[162,0.0014],[164,0.001],[166,0.0011],[170,0.001],[172,0.0011],[174,0.0015],[176,0.001],[178,0.0011],[180,0.0009],[188,0.0007],[196,0.0008],[200,0.0017],[204,0.0014],[206,0.0009],[210,0.0009],[212,0.0007],[214,0.0008],[216,0.0008],[222,0.0013],[226,0.0011],[228,0.0005],[230,0.0011],[232,0.0012],[234,0.0007],[236,0.0011],[238,0.0009],[244,0.0012],[246,0.0009],[248,0.0011],[250,0.0005],[252,0.0003],[254,0.0006],[256,0.0004],[258,0.0006],[264,0.0001],[266,0.0007],[268,0.0011],[272,0.0012],[274,0.0021],[276,0.0022],[278,0.0009],[280,0.0017],[282,0.0014],[286,0.0018],[290,0.0016],[292,0.0012],[294,0.0015],[296,0.0011],[298,0.001],[300,0.0012],[302,0.0015],[304,0.0013],[308,0.0016],[310,0.002],[312,0.0014],[316,0.0015],[318,0.0018],[320,0.0013],[324,0.0018],[326,0.0018],[328,0.0014],[332,0.0018],[336,0.0016],[338,0.0025],[340,0.0028],[342,0.0033],[344,0.0029],[346,0.0021],[348,0.0017],[354,0.0022],[356,0.0022],[358,0.0014],[362,0.0016],[368,0.0009],[372,0.0014],[376,0.0014],[378,0.0011],[380,0.0011],[382,0.0013],[384,0.0008],[386,0.002],[388,0.003],[390,0.003],[392,0.0025],[394,0.0017],[398,0.0027],[400,0.0027],[402,0.0024],[404,0.0032],[406,0.0029],[414,0.0025],[418,0.0029],[420,0.0028],[424,0.0028],[428,0.0023],[432,0.0026],[434,0.0022],[438,0.0024],[440,0.0027],[448,0.0023],[460,0.0027],[464,0.0025],[466,0.0026],[468,0.0025],[470,0.0013],[472,0.0028],[474,0.0031],[476,0.0031],[478,0.002],[480,0.0019],[484,0.0021],[486,0.0014],[488,0.0013],[490,0.0017],[492,0.0012],[494,0.0052],[496,0.0237],[498,0.0365],[500,0.028],[502,0.0225],[504,0.0067],[506,0.0029],[508,0.0019],[514,0.0004],[516,0.0025],[518,0.0032],[520,0.0079],[522,0.0051],[524,0.0039],[526,0.0037],[532,0.0038],[534,0.0057],[536,0.0084],[538,0.0085],[540,0.0084],[542,0.0093],[544,0.0055],[546,0.0052],[550,0.0064],[552,0.0062],[554,0.0068],[558,0.0085],[560,0.0089],[562,0.0085],[564,0.0091],[566,0.0086],[568,0.0084],[570,0.0081],[572,0.0058],[574,0.0057],[578,0.0051],[580,0.0052],[582,0.0056],[584,0.0077],[588,0.0138],[590,0.0115],[592,0.0075],[594,0.0068],[596,0.0026],[598,0.0021],[600,0.0024],[608,0.0023],[614,0.0028],[618,0.0024],[632,0.0022],[634,0.0019],[636,0.0018],[640,0.0022],[642,0.0022],[644,0.0016],[646,0.0015],[650,0.0023],[652,0.0024],[654,0.0023],[664,0.0024],[668,0.0022],[676,0.0024],[680,0.0022],[682,0.0019],[686,0.0023],[688,0.0021],[696,0.0022],[700,0.0017],[702,0.0022],[704,0.0021],[706,0.0024],[708,0.0022],[712,0.0025],[714,0.0023],[716,0.0024],[718,0.0021],[720,0.0021]]
       }
       CTRL_L_mouth_pressU {type=MovieSceneScriptingFloatChannel, default=0, const [0,492]}
       CTRL_L_mouth_funnelD {
        type=MovieSceneScriptingFloatChannel
        default=0
        thinned_from=492
        keys [[-262,0.0122],[-260,0.0157],[-258,0.0146],[-256,0.0152],[-254,0.017],[-252,0.0175],[-250,0.0186],[-248,0.021],[-246,0.0237],[-244,0.0244],[-242,0.0227],[-240,0.0221],[-238,0.021],[-236,0.0202],[-234,0.0197],[-232,0.0168],[-230,0.0174],[-226,0.0153],[-224,0.0153],[-222,0.0148],[-220,0.0128],[-218,0.0115],[-216,0.0116],[-214,0.0106],[-210,0.0102],[-208,0.0112],[-206,0.0095],[-204,0.01],[-202,0.0078],[-200,0.0063],[-198,0.0065],[-196,0.0062],[-194,0.0056],[-192,0.0052],[-190,0.0052],[-188,0.0076],[-186,0.0077],[-184,0.0084],[-182,0.0075],[-180,0.0069],[-178,0.0066],[-176,0.0058],[-174,0.0043],[-172,0.0026],[-170,0.0018],[-164,0.0013],[-162,0.0014],[-160,0.0057],[-158,0.0066],[-156,0.0146],[-154,0.0223],[-152,0.0282],[-150,0.0237],[-148,0.0159],[-146,0.0123],[-144,0.0103],[-142,0.0094],[-136,0.0077],[-134,0.006],[-132,0.0076],[-128,0.0065],[-126,0.0048],[-124,0.0035],[-122,0.0029],[-120,0.002],[-112,0.002],[-110,0.0017],[-108,0.0016],[-102,0.002],[-92,0.002],[-90,0.0023],[-82,0.0016],[-76,0.0019],[-74,0.0117],[-72,0.0099],[-70,0.0063],[-68,0.0033],[-66,0.0022],[-62,0.0021],[-60,0.0018],[-54,0.0018],[-46,0.0011],[-38,0.0017],[-36,0.0022],[-34,0.0019],[-30,0.0018],[-28,0.0021],[-26,0.0018],[-22,0.0022],[-18,0.0018],[-14,0.0021],[-8,0.002],[-6,0.0021],[2,0.002],[4,0.0021],[12,0.0018],[30,0.0017],[34,0.0019],[36,0.0018],[38,0.0019],[40,0.0018],[42,0.0019],[46,0.0018],[48,0.002],[66,0.0017],[70,0.0018],[72,0.0016],[80,0.0019],[90,0.0021],[92,0.0018],[100,0.0017],[102,0.0014],[104,0.0017],[106,0.0013],[112,0.0014],[116,0.0002],[126,0.0006],[130,0.0004],[132,0.0008],[136,0.0004],[138,0.0009],[144,0.0009],[146,0.0006],[148,0.0006],[150,0.0011],[152,0.0007],[156,0.0009],[160,0.0004],[162,0.0006],[164,0.0002],[166,0.0004],[168,0.0003],[170,0.0006],[172,0.0005],[174,0.0007],[176,0.0002],[180,0.0002],[182,0],[192,0.0001],[196,0],[200,0.0007],[204,0.0005],[206,0],[212,0],[214,0.0002],[216,0],[218,0.0002],[220,0],[222,0.0003],[226,0],[232,0.0002],[234,0],[266,0],[268,0.0003],[270,0.0001],[272,0.0002],[274,0.0009],[276,0.0009],[278,0.0003],[280,0.001],[282,0.0007],[284,0.0006],[286,0.0009],[290,0.0002],[292,0],[294,0.0005],[296,0.0001],[300,0],[302,0.0005],[306,0.0003],[310,0.0006],[312,0.0002],[316,0.0004],[318,0.0008],[320,0.0003],[324,0.0008],[326,0.0008],[328,0.0004],[330,0.0006],[332,0.0005],[334,0.0008],[336,0.0005],[338,0.0013],[342,0.0017],[344,0.0013],[346,0.0005],[350,0],[352,0.0003],[356,0.0006],[358,0.0002],[360,0],[362,0.0001],[382,0],[384,0.0001],[388,0.0015],[390,0.0014],[394,0.0003],[396,0.001],[398,0.0013],[400,0.0013],[402,0.001],[404,0.0018],[408,0.0013],[420,0.0012],[426,0.0014],[428,0.0011],[432,0.0014],[434,0.0011],[438,0.001],[440,0.0015],[444,0.0012],[446,0.0014],[450,0.0012],[454,0.0014],[464,0.001],[466,0.0011],[468,0.001],[470,0.0003],[472,0.0013],[474,0.0015],[476,0.0014],[478,0.0005],[484,0.0007],[486,0.0002],[488,0],[490,0.0003],[492,0.0002],[494,0.0037],[496,0.0203],[498,0.0323],[500,0.0248],[502,0.0194],[504,0.0052],[506,0.0019],[508,0.0006],[512,0],[514,0],[516,0.0015],[518,0.0022],[520,0.0069],[522,0.0039],[524,0.0027],[532,0.0029],[534,0.0046],[536,0.0067],[540,0.0065],[542,0.0073],[544,0.0036],[546,0.0034],[548,0.0036],[550,0.0043],[552,0.0041],[554,0.0047],[558,0.0064],[560,0.0068],[562,0.0063],[564,0.0067],[566,0.0061],[568,0.006],[570,0.0057],[572,0.0036],[578,0.0031],[580,0.0031],[582,0.0037],[584,0.0055],[586,0.0082],[588,0.0111],[590,0.009],[592,0.0055],[594,0.0048],[596,0.0014],[598,0.0011],[600,0.0014],[606,0.0012],[614,0.0018],[618,0.0015],[632,0.0013],[634,0.0009],[642,0.0014],[644,0.001],[646,0.0008],[648,0.0014],[654,0.0011],[660,0.0013],[662,0.0015],[666,0.0013],[670,0.0014],[674,0.0012],[676,0.0014],[678,0.0011],[680,0.0012],[684,0.0008],[686,0.0011],[688,0.0008],[692,0.0012],[696,0.0012],[700,0.0009],[702,0.0013],[704,0.0011],[720,0.0012]]
       }
       CTRL_L_mouth_suckBlow {
        type=MovieSceneScriptingFloatChannel
        default=0
        thinned_from=492
        keys [[-262,0.0049],[-260,0.0048],[-258,0.0042],[-256,0.0039],[-254,0.0041],[-252,0.0046],[-250,0.0049],[-204,0.0049],[-200,0.0039],[-198,0.0046],[-196,0.0049],[-146,0.0049],[-142,0.0039],[-138,0.0049],[-134,0.0039],[-126,0.004],[-124,0.0044],[-120,0.0045],[-118,0.0054],[-116,0.0098],[-114,0.0112],[-112,0.0144],[-110,0.016],[-108,0.0169],[-106,0.0171],[-104,0.017],[-102,0.015],[-100,0.0125],[-98,0.0122],[-94,0.0083],[-92,0.0077],[-90,0.0049],[-64,0.0049],[-62,0.0058],[-60,0.0117],[-58,0.0119],[-56,0.0124],[-54,0.0127],[-52,0.0127],[-50,0.0134],[-48,0.0137],[-46,0.0127],[-44,0.0077],[-42,0.005],[70,0.0049],[72,0.0054],[76,0.0049],[80,0.0049],[82,0.0052],[84,0.0063],[86,0.0063],[88,0.0055],[90,0.0054],[92,0.0058],[96,0.0049],[274,0.0049],[276,0.0088],[278,0.0088],[280,0.0051],[480,0.0049],[482,0.0045],[490,0.0044],[492,0.0049],[720,0.0049]]
       }
       CTRL_L_mouth_pressD {type=MovieSceneScriptingFloatChannel, default=0, const [0,492]}
       CTRL_L_mouth_funnelU {
        type=MovieSceneScriptingFloatChannel
        default=0
        thinned_from=492
        keys [[-262,0.0024],[-260,0.0012],[-258,0.001],[-252,0],[-248,0.0005],[-242,0],[-240,0.0005],[-238,0.0002],[-236,0.0007],[-230,0.001],[-226,0.0008],[-222,0.0009],[-218,0.0003],[-214,0.0014],[-212,0.0015],[-210,0.0012],[-208,0.0011],[-206,0.0008],[-202,0.0009],[-200,0.0005],[-198,0.0009],[-190,0.0012],[-188,0.0008],[-182,0.0005],[-178,0.0012],[-176,0.0015],[-174,0.0021],[-160,0.0029],[-154,0.0026],[-152,0.0029],[-150,0.0026],[-148,0.0025],[-144,0.0027],[-142,0.0025],[-140,0.0028],[-138,0.0026],[-134,0.0028],[-130,0.0025],[-122,0.0026],[-116,0.0029],[-108,0.0027],[-102,0.0028],[-98,0.003],[-86,0.0028],[-84,0.0026],[-76,0.0028],[-74,0.0026],[-70,0.0027],[-68,0.0025],[-64,0.0028],[-56,0.0024],[-52,0.0024],[-48,0.0021],[-40,0.0028],[-38,0.0026],[-36,0.003],[-30,0.0025],[-16,0.0028],[-4,0.0025],[4,0.0027],[6,0.0025],[10,0.0029],[18,0.0029],[24,0.0028],[32,0.0029],[40,0.0029],[56,0.0027],[58,0.0029],[66,0.0027],[86,0.003],[90,0.0028],[92,0.0029],[94,0.0028],[102,0.0026],[104,0.0028],[106,0.0026],[108,0.0027],[112,0.0024],[114,0.0017],[116,0.0007],[118,0.0011],[124,0.0017],[130,0.0016],[132,0.0022],[136,0.0014],[138,0.0019],[144,0.0017],[146,0.0014],[148,0.0016],[150,0.002],[152,0.0017],[158,0.0015],[160,0.0013],[162,0.0016],[164,0.0014],[166,0.0014],[170,0.001],[172,0.0013],[174,0.0014],[176,0.0008],[178,0.001],[182,0.0008],[184,0.0011],[188,0.0006],[196,0.0005],[200,0.0014],[204,0.0013],[206,0.0007],[210,0.0009],[212,0.0007],[214,0.001],[216,0.0005],[222,0.0011],[224,0.0009],[226,0.0009],[228,0.0005],[230,0.0011],[232,0.0012],[234,0.0006],[244,0.001],[246,0.0008],[248,0.0009],[250,0.0004],[254,0.0009],[256,0.0004],[260,0.0007],[264,0.0002],[266,0.0005],[268,0.0013],[272,0.0013],[274,0.0021],[276,0.0088],[278,0.0076],[280,0.0096],[282,0.0097],[284,0.0111],[286,0.011],[288,0.0098],[290,0.0069],[292,0.0079],[294,0.0076],[296,0.009],[298,0.0054],[300,0.0069],[302,0.0089],[304,0.0085],[306,0.0104],[308,0.0108],[310,0.008],[312,0.0078],[314,0.008],[316,0.0069],[318,0.0074],[320,0.0072],[322,0.0081],[324,0.0078],[326,0.0072],[328,0.0055],[330,0.0059],[332,0.0043],[334,0.0029],[336,0.0021],[338,0.0026],[342,0.0027],[344,0.0024],[346,0.0015],[348,0.0009],[356,0.0017],[358,0.0009],[360,0.0008],[364,0.0009],[366,0.0005],[368,0.0006],[370,0.0009],[372,0.0009],[376,0.0011],[380,0.0006],[384,0.0006],[386,0.0018],[388,0.0028],[390,0.0028],[394,0.0018],[396,0.0025],[398,0.0028],[400,0.0028],[402,0.0023],[404,0.003],[408,0.0023],[414,0.0024],[418,0.0022],[422,0.0024],[428,0.0023],[432,0.0024],[434,0.0021],[438,0.0022],[440,0.0027],[444,0.0021],[446,0.0023],[452,0.0023],[454,0.0025],[462,0.002],[464,0.0021],[466,0.0025],[468,0.0023],[470,0.001],[472,0.0022],[474,0.0026],[476,0.0025],[478,0.0015],[480,0.0014],[484,0.0016],[488,0.0009],[490,0.0012],[492,0.0009],[494,0.0008],[496,0.0003],[498,0.0006],[500,0.0002],[502,0.0008],[504,0.0004],[506,0.0005],[508,0.001],[510,0.0011],[512,0.0007],[514,0],[518,0.0002],[520,0.0014],[522,0.002],[524,0.0019],[526,0.0028],[530,0.0031],[536,0.0029],[538,0.0027],[542,0.0025],[546,0.0015],[548,0.0007],[558,0.0003],[562,0.0006],[564,0.0004],[566,0.0007],[570,0.0007],[572,0.0001],[576,0.0006],[582,0.0002],[586,0.0007],[588,0.0007],[592,0.0013],[596,0.0015],[600,0.0023],[608,0.0023],[614,0.0027],[618,0.0024],[620,0.0024],[622,0.002],[628,0.002],[632,0.0021],[636,0.0018],[640,0.0021],[642,0.002],[646,0.0015],[648,0.002],[652,0.0022],[654,0.0019],[656,0.0021],[662,0.002],[664,0.0021],[666,0.0019],[670,0.0021],[672,0.002],[676,0.0022],[682,0.0018],[686,0.0021],[692,0.0018],[694,0.0019],[700,0.0016],[702,0.002],[704,0.002],[706,0.0021],[714,0.0018],[716,0.0019],[720,0.0018]]
       }
       CTRL_R_mouth_towardsU {
        type=MovieSceneScriptingFloatChannel
        default=0
        thinned_from=492
        keys [[-262,0.0144],[-260,0.0144],[-258,0.0142],[-256,0.0136],[-254,0.014],[-252,0.0137],[-248,0.0153],[-246,0.0147],[-244,0.0145],[-242,0.0141],[-240,0.0146],[-234,0.0144],[-232,0.0141],[-230,0.0144],[-228,0.0151],[-226,0.0153],[-222,0.0152],[-220,0.0146],[-218,0.0146],[-216,0.0143],[-212,0.0141],[-210,0.0131],[-208,0.014],[-206,0.014],[-202,0.0147],[-200,0.0136],[-198,0.0143],[-196,0.0147],[-194,0.0148],[-190,0.0147],[-188,0.0153],[-186,0.0152],[-184,0.0146],[-182,0.0147],[-180,0.0154],[-178,0.0164],[-176,0.0171],[-174,0.017],[-172,0.016],[-168,0.0126],[-160,0.0078],[-158,0.0077],[-156,0.0071],[-154,0.0061],[-152,0.006],[-150,0.0062],[-148,0.0069],[-146,0.0085],[-144,0.0089],[-140,0.0088],[-138,0.0096],[-136,0.0094],[-134,0.01],[-130,0.0097],[-126,0.0106],[-122,0.012],[-120,0.0132],[-118,0.0153],[-116,0.0193],[-114,0.0198],[-112,0.021],[-110,0.0208],[-106,0.0197],[-104,0.0194],[-102,0.0197],[-100,0.0194],[-98,0.0194],[-96,0.0185],[-94,0.0166],[-92,0.0163],[-90,0.0137],[-88,0.013],[-86,0.0106],[-84,0.0096],[-82,0.0089],[-78,0.0085],[-74,0.0088],[-72,0.0095],[-70,0.0104],[-68,0.0103],[-66,0.0113],[-64,0.0141],[-62,0.0148],[-60,0.0172],[-48,0.0176],[-46,0.0172],[-44,0.0197],[-42,0.0193],[-40,0.0179],[-38,0.0145],[-36,0.0127],[-34,0.0133],[-32,0.0132],[-30,0.0135],[-26,0.0132],[-22,0.0136],[-20,0.0128],[-18,0.0136],[-16,0.0136],[-14,0.0132],[-12,0.0133],[-8,0.0127],[-6,0.0125],[-4,0.0131],[-2,0.0135],[0,0.013],[2,0.0136],[4,0.0137],[6,0.0134],[10,0.0132],[12,0.0125],[14,0.0128],[18,0.0128],[20,0.0126],[26,0.013],[30,0.0127],[32,0.0124],[34,0.0125],[38,0.0139],[40,0.0141],[42,0.014],[44,0.0133],[56,0.0131],[60,0.0134],[64,0.0132],[66,0.0126],[70,0.0125],[74,0.0126],[84,0.0137],[90,0.0136],[92,0.0145],[94,0.0141],[96,0.0155],[98,0.0154],[100,0.0144],[102,0.0154],[104,0.0167],[106,0.0178],[108,0.0181],[110,0.018],[112,0.0174],[114,0.0163],[116,0.0146],[118,0.0142],[120,0.0143],[122,0.0152],[126,0.0154],[128,0.016],[130,0.016],[132,0.0171],[134,0.0166],[136,0.0156],[138,0.0164],[142,0.016],[144,0.0162],[146,0.0153],[150,0.0169],[152,0.0164],[156,0.0163],[160,0.0152],[162,0.0162],[164,0.016],[168,0.0149],[172,0.0157],[174,0.0153],[176,0.0138],[180,0.0141],[182,0.0139],[184,0.0146],[188,0.0145],[190,0.0142],[192,0.0141],[196,0.0133],[200,0.0144],[202,0.0138],[204,0.0135],[210,0.0136],[212,0.0141],[214,0.0143],[216,0.0137],[218,0.0138],[220,0.0145],[222,0.0145],[224,0.0147],[226,0.0137],[228,0.0134],[230,0.0145],[232,0.0141],[236,0.0157],[238,0.0146],[240,0.0147],[242,0.0145],[244,0.0149],[246,0.0148],[248,0.015],[250,0.0144],[254,0.0154],[256,0.0152],[258,0.0155],[260,0.0166],[262,0.0156],[264,0.0153],[268,0.0178],[272,0.0199],[274,0.0213],[276,0.0198],[278,0.0154],[280,0.0166],[284,0.0168],[286,0.0174],[288,0.0172],[290,0.016],[292,0.0153],[294,0.0164],[296,0.0157],[298,0.0147],[302,0.016],[306,0.0159],[310,0.0163],[312,0.0154],[314,0.0155],[316,0.0148],[318,0.0162],[320,0.0151],[322,0.0149],[324,0.015],[326,0.0139],[330,0.0139],[332,0.0132],[338,0.013],[342,0.0122],[344,0.0123],[348,0.0115],[350,0.0118],[352,0.0111],[356,0.0115],[358,0.0101],[360,0.01],[362,0.0112],[364,0.0113],[366,0.0108],[368,0.0101],[372,0.0111],[374,0.0124],[378,0.0128],[380,0.0123],[382,0.0124],[384,0.0113],[386,0.0126],[388,0.0146],[390,0.0132],[392,0.009],[394,0.0063],[396,0.0075],[398,0.0081],[400,0.0079],[402,0.0074],[404,0.0083],[406,0.0082],[410,0.0099],[412,0.011],[416,0.0123],[424,0.0127],[428,0.0137],[432,0.0137],[434,0.0142],[436,0.0143],[438,0.0142],[440,0.0146],[442,0.014],[444,0.0142],[446,0.0139],[448,0.0143],[450,0.0141],[454,0.0143],[456,0.0137],[460,0.0116],[464,0.0099],[466,0.0084],[468,0.0082],[470,0.0071],[472,0.0081],[474,0.0099],[476,0.011],[478,0.0094],[480,0.0101],[482,0.0123],[484,0.0127],[486,0.0129],[488,0.0136],[492,0.0114],[494,0.0132],[496,0.0156],[498,0.016],[500,0.0147],[502,0.015],[504,0.0127],[506,0.0113],[508,0.0125],[514,0.0125],[516,0.015],[518,0.0159],[520,0.0237],[522,0.0247],[526,0.0233],[528,0.0232],[530,0.0228],[532,0.0227],[534,0.0206],[536,0.019],[538,0.0139],[540,0.0129],[542,0.0101],[544,0.0114],[546,0.0119],[548,0.0118],[550,0.0128],[558,0.0136],[560,0.0143],[564,0.014],[566,0.0144],[570,0.0145],[572,0.0126],[576,0.013],[580,0.0125],[582,0.0121],[584,0.012],[586,0.013],[588,0.0126],[590,0.0114],[592,0.0108],[594,0.0107],[596,0.0091],[598,0.0095],[600,0.0113],[602,0.0115],[604,0.0128],[606,0.0133],[608,0.0148],[610,0.0156],[612,0.0162],[616,0.0164],[620,0.016],[622,0.0154],[624,0.0155],[626,0.0159],[628,0.0165],[630,0.0167],[634,0.0164],[636,0.0156],[638,0.0159],[640,0.0161],[644,0.0157],[646,0.0153],[648,0.016],[652,0.0169],[656,0.017],[660,0.0163],[662,0.0165],[664,0.0171],[666,0.0165],[668,0.0167],[670,0.0166],[672,0.017],[674,0.017],[678,0.0169],[680,0.0165],[682,0.0158],[688,0.0161],[692,0.0159],[694,0.0156],[696,0.0156],[698,0.016],[700,0.0157],[702,0.0158],[704,0.0163],[706,0.0165],[708,0.0162],[712,0.0164],[716,0.017],[718,0.0164],[720,0.0164]]
       }
       CTRL_R_mouth_purseD {
        type=MovieSceneScriptingFloatChannel
        default=0
        thinned_from=492
        keys [[-262,0.0662],[-260,0.0658],[-258,0.057],[-256,0.0521],[-254,0.0515],[-252,0.0512],[-250,0.0529],[-248,0.0534],[-246,0.0541],[-244,0.0544],[-242,0.055],[-240,0.0573],[-238,0.0581],[-236,0.0566],[-234,0.0558],[-232,0.0504],[-230,0.0485],[-226,0.046],[-224,0.0438],[-222,0.0414],[-218,0.0411],[-216,0.0387],[-214,0.0352],[-212,0.0346],[-210,0.0335],[-208,0.0357],[-206,0.0359],[-204,0.0358],[-202,0.037],[-200,0.0349],[-198,0.0381],[-196,0.0377],[-194,0.0364],[-190,0.0374],[-188,0.0461],[-186,0.0465],[-184,0.0459],[-182,0.0505],[-180,0.0568],[-178,0.0624],[-176,0.0724],[-174,0.0899],[-172,0.1037],[-170,0.1188],[-168,0.1358],[-166,0.1572],[-164,0.209],[-162,0.2589],[-160,0.2979],[-158,0.3033],[-156,0.3498],[-154,0.3553],[-152,0.3675],[-150,0.3761],[-148,0.3723],[-146,0.3791],[-144,0.3828],[-142,0.3765],[-138,0.3718],[-136,0.3643],[-134,0.3583],[-132,0.3651],[-130,0.3595],[-128,0.3523],[-126,0.3401],[-124,0.3271],[-122,0.3065],[-120,0.259],[-118,0.1836],[-116,0.0996],[-114,0.0831],[-112,0.0489],[-110,0.0343],[-108,0.0258],[-106,0.0208],[-104,0.0184],[-100,0.017],[-98,0.0168],[-96,0.0205],[-92,0.0257],[-90,0.0436],[-88,0.0519],[-86,0.0832],[-84,0.1229],[-82,0.16],[-80,0.2275],[-78,0.2942],[-76,0.3097],[-74,0.3365],[-72,0.3293],[-70,0.3281],[-68,0.3118],[-66,0.2685],[-64,0.179],[-62,0.157],[-60,0.06],[-58,0.0522],[-56,0.0353],[-54,0.0252],[-52,0.0206],[-50,0.0169],[-48,0.0157],[-46,0.0177],[-44,0.0279],[-42,0.0324],[-40,0.0261],[-38,0.0264],[-36,0.0328],[-34,0.0269],[-32,0.0233],[-30,0.025],[-28,0.0307],[-26,0.03],[-24,0.0269],[-22,0.0249],[-20,0.0216],[-18,0.0217],[-16,0.0211],[-14,0.0179],[-12,0.0167],[-8,0.0162],[-6,0.0163],[-4,0.0146],[-2,0.0152],[2,0.0146],[6,0.0146],[8,0.0148],[10,0.0155],[12,0.0159],[14,0.0156],[16,0.0158],[18,0.0164],[20,0.0164],[24,0.0156],[26,0.0162],[28,0.0165],[30,0.0171],[34,0.0177],[36,0.0165],[38,0.0176],[40,0.0205],[42,0.019],[44,0.0184],[54,0.0171],[56,0.0164],[58,0.0172],[60,0.0165],[62,0.0155],[64,0.0168],[66,0.0168],[70,0.0172],[72,0.0182],[74,0.0167],[78,0.0177],[80,0.0175],[82,0.0189],[84,0.0198],[86,0.0185],[88,0.0179],[90,0.0185],[92,0.0159],[94,0.018],[96,0.0214],[98,0.0219],[100,0.0212],[102,0.0245],[104,0.0246],[106,0.02],[108,0.0189],[110,0.0171],[112,0.0174],[114,0.0162],[116,0.0159],[118,0.0174],[120,0.0177],[122,0.0173],[124,0.0177],[126,0.0176],[128,0.0164],[130,0.0164],[134,0.0151],[136,0.0161],[138,0.0159],[140,0.0165],[142,0.0155],[144,0.0156],[146,0.0145],[148,0.0146],[150,0.0159],[152,0.0164],[154,0.0152],[156,0.0152],[158,0.0138],[160,0.0134],[162,0.0146],[164,0.014],[166,0.0142],[168,0.0139],[170,0.0142],[172,0.0132],[176,0.0117],[178,0.0121],[180,0.0114],[182,0.0129],[188,0.0124],[190,0.0124],[192,0.0126],[194,0.0118],[200,0.0124],[204,0.0121],[206,0.0107],[208,0.0112],[210,0.0113],[212,0.0123],[214,0.0127],[216,0.0114],[218,0.0124],[220,0.0131],[226,0.0119],[228,0.0118],[230,0.0125],[232,0.0122],[234,0.0131],[236,0.0162],[238,0.0153],[240,0.0164],[242,0.0143],[244,0.0136],[246,0.0155],[248,0.0161],[250,0.0131],[252,0.014],[254,0.0136],[256,0.0143],[258,0.0126],[260,0.0125],[262,0.0116],[264,0.0119],[266,0.012],[268,0.0135],[270,0.0159],[272,0.0214],[274,0.0228],[276,0.0208],[278,0.0158],[280,0.0166],[282,0.0165],[284,0.019],[286,0.0204],[288,0.0232],[290,0.0229],[292,0.0271],[294,0.0282],[296,0.0243],[298,0.0243],[300,0.0233],[302,0.0251],[304,0.0232],[308,0.0234],[310,0.0231],[312,0.0231],[314,0.0216],[316,0.0209],[320,0.0216],[322,0.0234],[324,0.0203],[326,0.0201],[328,0.0212],[330,0.0199],[332,0.0174],[334,0.0191],[336,0.0232],[338,0.0241],[340,0.0227],[342,0.0209],[344,0.0222],[346,0.0221],[348,0.0287],[350,0.0308],[352,0.0291],[354,0.0292],[356,0.029],[358,0.0277],[360,0.0283],[362,0.0347],[364,0.0386],[366,0.0396],[368,0.0319],[370,0.03],[372,0.0352],[374,0.0381],[376,0.0384],[378,0.0384],[380,0.0377],[382,0.0352],[384,0.0371],[386,0.0308],[388,0.0284],[390,0.0282],[392,0.0328],[394,0.0416],[396,0.0518],[398,0.0648],[400,0.0687],[402,0.077],[404,0.061],[406,0.0485],[408,0.0414],[410,0.0394],[412,0.0397],[414,0.0395],[416,0.0376],[418,0.0361],[420,0.0325],[422,0.0297],[424,0.0296],[426,0.0305],[428,0.0307],[430,0.0301],[432,0.0318],[434,0.032],[436,0.0336],[438,0.0338],[440,0.0353],[442,0.0342],[444,0.0356],[446,0.0393],[448,0.0396],[450,0.0378],[452,0.0386],[454,0.041],[456,0.042],[458,0.0393],[460,0.041],[462,0.0419],[464,0.0446],[466,0.0507],[468,0.0527],[470,0.0617],[472,0.0642],[474,0.0722],[476,0.0723],[480,0.0769],[482,0.0895],[484,0.0925],[486,0.0894],[488,0.1057],[490,0.1117],[492,0.1386],[494,0.1554],[496,0.1525],[498,0.125],[500,0.1048],[502,0.1052],[504,0.0921],[506,0.0883],[508,0.0938],[510,0.0886],[512,0.0881],[514,0.0964],[516,0.1159],[518,0.1155],[520,0.1097],[522,0.0988],[524,0.0782],[526,0.0529],[528,0.0519],[530,0.0506],[532,0.05],[536,0.0441],[538,0.043],[540,0.0429],[542,0.0451],[544,0.0515],[546,0.0606],[548,0.0629],[550,0.0691],[552,0.0679],[554,0.0653],[556,0.0639],[558,0.0627],[560,0.0628],[562,0.0617],[564,0.0627],[566,0.0619],[568,0.0618],[570,0.0631],[572,0.0596],[576,0.0623],[578,0.0621],[580,0.0628],[582,0.0602],[584,0.0543],[586,0.0514],[588,0.0489],[590,0.0525],[592,0.0638],[594,0.0656],[596,0.0669],[598,0.0644],[600,0.0599],[602,0.0572],[604,0.036],[606,0.0339],[608,0.0278],[610,0.0255],[614,0.0249],[616,0.0238],[618,0.0247],[620,0.0254],[622,0.0302],[624,0.0309],[626,0.031],[628,0.0302],[632,0.0262],[634,0.0267],[636,0.0224],[638,0.0233],[640,0.026],[642,0.0252],[644,0.0253],[646,0.0277],[648,0.0293],[650,0.0285],[652,0.0331],[654,0.0356],[656,0.0386],[658,0.0374],[660,0.0407],[662,0.0393],[664,0.0371],[666,0.0367],[668,0.0351],[670,0.0377],[672,0.0366],[674,0.0339],[676,0.036],[678,0.0343],[680,0.0363],[682,0.0332],[684,0.0347],[686,0.0339],[688,0.0314],[692,0.0359],[694,0.0349],[696,0.0322],[698,0.0312],[700,0.0299],[702,0.0272],[704,0.0262],[706,0.0287],[708,0.0271],[710,0.0274],[712,0.0279],[714,0.0301],[716,0.0299],[718,0.0275],[720,0.0275]]
       }
       CTRL_L_mouth_lipBiteU {
        type=MovieSceneScriptingFloatChannel
        default=0
        thinned_from=492
        keys [[-262,0.0422],[-260,0.0414],[-258,0.0379],[-256,0.0364],[-254,0.0358],[-252,0.0356],[-250,0.0358],[-248,0.0357],[-246,0.0363],[-244,0.0365],[-242,0.0371],[-240,0.0371],[-238,0.0369],[-236,0.0364],[-234,0.0365],[-232,0.037],[-230,0.0351],[-228,0.0319],[-226,0.0294],[-224,0.0277],[-222,0.0266],[-220,0.026],[-218,0.0258],[-214,0.0244],[-212,0.0243],[-210,0.0239],[-208,0.0243],[-206,0.0238],[-204,0.0236],[-202,0.0215],[-200,0.0214],[-198,0.0231],[-196,0.023],[-192,0.0211],[-190,0.021],[-188,0.0255],[-186,0.0258],[-184,0.0268],[-182,0.0291],[-180,0.034],[-178,0.0432],[-176,0.0533],[-174,0.0676],[-172,0.0855],[-170,0.1059],[-168,0.1255],[-166,0.1391],[-164,0.1579],[-162,0.1801],[-160,0.2079],[-158,0.2119],[-156,0.2476],[-154,0.2881],[-152,0.3169],[-150,0.3589],[-148,0.3882],[-146,0.3989],[-144,0.4115],[-140,0.4229],[-138,0.4329],[-136,0.4318],[-134,0.4289],[-132,0.4303],[-130,0.4298],[-128,0.4254],[-126,0.4175],[-124,0.4143],[-122,0.4038],[-120,0.3755],[-118,0.3232],[-116,0.2234],[-114,0.2075],[-112,0.1692],[-110,0.142],[-108,0.1256],[-106,0.1146],[-104,0.1094],[-102,0.0902],[-100,0.0827],[-98,0.082],[-96,0.0749],[-94,0.0828],[-92,0.086],[-90,0.1011],[-88,0.1037],[-86,0.1138],[-84,0.1299],[-82,0.1465],[-80,0.171],[-78,0.207],[-76,0.2188],[-74,0.3113],[-72,0.3521],[-70,0.3789],[-68,0.3904],[-66,0.3772],[-64,0.3109],[-62,0.2843],[-60,0.1531],[-58,0.1449],[-56,0.1261],[-54,0.113],[-52,0.1058],[-50,0.1007],[-48,0.0992],[-46,0.0913],[-44,0.0724],[-42,0.0637],[-40,0.0764],[-38,0.0922],[-36,0.098],[-34,0.0942],[-32,0.0842],[-30,0.0784],[-28,0.0762],[-26,0.078],[-24,0.0762],[-22,0.0722],[-20,0.0732],[-18,0.0699],[-16,0.0643],[-14,0.0644],[-12,0.0639],[-10,0.0607],[-8,0.0586],[-6,0.058],[-4,0.058],[-2,0.0536],[0,0.052],[2,0.0524],[4,0.0518],[6,0.0522],[8,0.0522],[10,0.0539],[12,0.0533],[14,0.0535],[16,0.0529],[20,0.0487],[22,0.0493],[24,0.0523],[26,0.055],[28,0.0548],[30,0.0542],[32,0.0556],[34,0.0578],[36,0.0566],[38,0.0551],[40,0.0576],[42,0.059],[44,0.0584],[46,0.058],[48,0.0574],[52,0.0597],[54,0.0612],[56,0.0617],[60,0.0607],[62,0.0598],[64,0.0591],[66,0.0615],[68,0.0618],[70,0.0614],[72,0.0585],[76,0.0617],[78,0.0616],[82,0.0583],[84,0.0573],[86,0.0592],[88,0.0617],[90,0.062],[92,0.0534],[94,0.0547],[96,0.0605],[98,0.061],[100,0.0646],[102,0.0644],[104,0.0649],[106,0.0662],[108,0.0638],[110,0.0568],[112,0.0543],[114,0.0542],[116,0.0503],[118,0.0474],[122,0.0466],[124,0.0464],[126,0.0459],[128,0.0387],[130,0.0387],[132,0.039],[134,0.0419],[138,0.0407],[140,0.0406],[144,0.0373],[146,0.0396],[148,0.0407],[150,0.0437],[152,0.0446],[154,0.04],[156,0.0399],[158,0.0382],[160,0.0393],[162,0.0395],[166,0.0373],[168,0.0389],[170,0.0391],[172,0.0354],[174,0.0349],[176,0.0366],[180,0.0354],[182,0.0365],[184,0.038],[186,0.0376],[190,0.0351],[192,0.0365],[194,0.0373],[196,0.0372],[198,0.0373],[200,0.0356],[202,0.035],[204,0.035],[206,0.0344],[208,0.0341],[210,0.0341],[212,0.0337],[216,0.0334],[218,0.0338],[220,0.0354],[222,0.0357],[224,0.0342],[226,0.0313],[228,0.0296],[230,0.0307],[232,0.0321],[234,0.0318],[238,0.0357],[240,0.037],[242,0.0359],[244,0.0334],[246,0.0329],[248,0.033],[250,0.0328],[252,0.0325],[254,0.0316],[256,0.0306],[258,0.0303],[260,0.0284],[264,0.0275],[266,0.026],[268,0.0264],[270,0.0281],[274,0.0287],[276,0.0265],[278,0.0228],[280,0.0234],[284,0.0251],[286,0.0251],[288,0.0247],[290,0.0248],[294,0.0239],[296,0.0242],[298,0.0252],[300,0.0255],[302,0.0249],[306,0.0248],[308,0.0249],[310,0.0254],[312,0.0249],[314,0.0238],[316,0.0236],[318,0.0244],[322,0.0241],[324,0.0245],[328,0.024],[330,0.024],[334,0.0253],[336,0.0266],[338,0.0311],[340,0.0333],[342,0.0357],[344,0.0377],[348,0.0375],[350,0.0376],[352,0.0407],[356,0.0412],[358,0.0398],[360,0.0397],[362,0.0411],[364,0.0417],[366,0.04],[368,0.0381],[370,0.0364],[372,0.0362],[374,0.034],[376,0.0335],[378,0.0312],[380,0.0305],[382,0.03],[384,0.031],[386,0.029],[388,0.0287],[390,0.0312],[392,0.0409],[394,0.0491],[396,0.0529],[398,0.0584],[400,0.0603],[402,0.0613],[404,0.0515],[406,0.0476],[408,0.052],[410,0.0476],[412,0.0455],[414,0.0453],[416,0.0447],[418,0.045],[420,0.0443],[422,0.0423],[424,0.042],[426,0.0422],[428,0.0409],[430,0.0405],[432,0.0429],[434,0.0431],[436,0.0423],[438,0.0433],[440,0.0448],[442,0.0449],[444,0.046],[446,0.0483],[448,0.0489],[450,0.0477],[452,0.0479],[456,0.049],[458,0.0492],[460,0.0505],[462,0.0525],[464,0.0523],[468,0.0535],[470,0.0508],[472,0.0476],[474,0.0472],[476,0.0486],[478,0.0571],[480,0.0604],[482,0.0631],[484,0.0637],[488,0.0629],[490,0.065],[492,0.0723],[494,0.0805],[496,0.0815],[498,0.0755],[500,0.068],[502,0.0666],[504,0.0677],[506,0.0704],[508,0.0749],[510,0.0712],[512,0.0653],[514,0.0624],[516,0.053],[520,0.0527],[522,0.057],[524,0.0576],[526,0.0546],[528,0.057],[530,0.0607],[532,0.0592],[534,0.0546],[536,0.0472],[538,0.0429],[540,0.0425],[542,0.0508],[544,0.0517],[546,0.0492],[548,0.0476],[552,0.0471],[554,0.0467],[556,0.047],[566,0.0459],[568,0.0459],[570,0.0455],[576,0.0477],[578,0.0476],[580,0.0472],[582,0.0456],[584,0.0413],[586,0.0359],[588,0.0347],[590,0.0392],[592,0.0462],[594,0.0474],[596,0.0521],[598,0.051],[600,0.0399],[602,0.0392],[604,0.0352],[606,0.0349],[608,0.0344],[610,0.0346],[614,0.0372],[616,0.0378],[618,0.0371],[620,0.0369],[622,0.039],[624,0.0391],[626,0.0376],[628,0.0364],[634,0.0358],[636,0.0358],[638,0.0355],[640,0.0361],[642,0.037],[644,0.0359],[646,0.0352],[648,0.0354],[650,0.0353],[652,0.0362],[654,0.0377],[656,0.0378],[658,0.0376],[660,0.0385],[662,0.0389],[664,0.0378],[666,0.0378],[668,0.0375],[670,0.0375],[672,0.0383],[674,0.038],[676,0.0384],[678,0.0393],[680,0.0397],[682,0.0394],[684,0.0395],[686,0.04],[688,0.0391],[690,0.0398],[692,0.0402],[694,0.0386],[698,0.0394],[700,0.039],[704,0.0393],[706,0.0406],[708,0.0402],[710,0.0395],[712,0.0394],[716,0.0398],[718,0.0392],[720,0.0392]]
       }
       CTRL_R_mouth_pressD {type=MovieSceneScriptingFloatChannel, default=0, const [0,492]}
       CTRL_R_mouth_pressU {type=MovieSceneScriptingFloatChannel, default=0, const [0,492]}
       CTRL_L_mouth_purseD {
        type=MovieSceneScriptingFloatChannel
        default=0
        thinned_from=492
        keys [[-262,0.0843],[-260,0.0834],[-258,0.0753],[-256,0.0699],[-254,0.0683],[-252,0.0671],[-250,0.0676],[-248,0.0668],[-244,0.0671],[-242,0.0675],[-240,0.0693],[-238,0.0702],[-236,0.0683],[-234,0.0674],[-232,0.0608],[-230,0.0587],[-226,0.0565],[-222,0.0528],[-220,0.0527],[-218,0.0521],[-216,0.0499],[-214,0.0469],[-212,0.0464],[-210,0.0452],[-208,0.0476],[-206,0.0475],[-204,0.0477],[-202,0.0494],[-200,0.0466],[-198,0.0498],[-196,0.0494],[-194,0.0485],[-190,0.0495],[-188,0.0574],[-184,0.0582],[-182,0.0626],[-180,0.0683],[-178,0.0731],[-176,0.082],[-174,0.0979],[-172,0.1102],[-170,0.1242],[-168,0.1401],[-166,0.1613],[-164,0.2134],[-162,0.2658],[-160,0.3071],[-158,0.3123],[-156,0.3572],[-154,0.3613],[-152,0.3698],[-150,0.373],[-148,0.3702],[-144,0.3801],[-142,0.3751],[-140,0.3735],[-138,0.371],[-134,0.3575],[-132,0.364],[-130,0.3596],[-128,0.353],[-124,0.3295],[-122,0.3111],[-120,0.2696],[-118,0.205],[-116,0.13],[-114,0.1127],[-112,0.0758],[-110,0.0582],[-108,0.0486],[-106,0.0432],[-104,0.0407],[-102,0.0391],[-100,0.0381],[-98,0.0379],[-96,0.0433],[-94,0.0474],[-92,0.0507],[-90,0.0703],[-88,0.0779],[-86,0.1065],[-84,0.1424],[-82,0.1766],[-80,0.2383],[-78,0.301],[-76,0.3159],[-74,0.3361],[-72,0.33],[-70,0.3288],[-68,0.3143],[-66,0.2789],[-64,0.2005],[-62,0.1801],[-60,0.0892],[-58,0.0805],[-56,0.0612],[-54,0.0498],[-52,0.0445],[-50,0.0401],[-48,0.0388],[-46,0.0423],[-44,0.0508],[-42,0.0571],[-40,0.0522],[-38,0.0532],[-36,0.0597],[-34,0.0519],[-32,0.0483],[-30,0.0482],[-28,0.0533],[-26,0.0517],[-24,0.0481],[-22,0.0458],[-20,0.0422],[-16,0.0406],[-14,0.0368],[-12,0.0353],[-8,0.0363],[-6,0.0366],[-4,0.0339],[-2,0.0355],[0,0.0343],[2,0.0344],[6,0.036],[8,0.0371],[10,0.0372],[12,0.0381],[16,0.0386],[20,0.0398],[24,0.0381],[26,0.0384],[28,0.0394],[30,0.04],[32,0.0413],[34,0.0413],[36,0.0394],[38,0.0404],[40,0.044],[42,0.0415],[44,0.0409],[50,0.0407],[54,0.0401],[56,0.0401],[58,0.0415],[60,0.0409],[62,0.0394],[64,0.0411],[66,0.0413],[68,0.0408],[70,0.0411],[72,0.0428],[74,0.0409],[78,0.0389],[80,0.0398],[82,0.0418],[84,0.042],[86,0.0407],[88,0.0404],[90,0.0411],[92,0.0399],[94,0.0413],[96,0.0442],[98,0.0452],[100,0.045],[102,0.0474],[104,0.0472],[106,0.041],[110,0.0387],[112,0.0396],[116,0.038],[118,0.0391],[122,0.0396],[124,0.0393],[126,0.0393],[128,0.0384],[130,0.0384],[132,0.0377],[134,0.0373],[136,0.0395],[138,0.0392],[140,0.0394],[142,0.0392],[144,0.0387],[146,0.0379],[148,0.0378],[150,0.0393],[152,0.0387],[154,0.037],[156,0.0369],[158,0.0355],[160,0.0356],[162,0.0361],[164,0.0355],[168,0.0353],[170,0.0356],[172,0.0342],[176,0.0332],[178,0.033],[180,0.0327],[182,0.0343],[184,0.0339],[190,0.0343],[192,0.034],[194,0.0329],[196,0.0327],[198,0.033],[200,0.0335],[202,0.0336],[204,0.0334],[206,0.0318],[210,0.0319],[214,0.0332],[216,0.0322],[218,0.0325],[224,0.0325],[226,0.0328],[228,0.0329],[230,0.0338],[232,0.0339],[234,0.0347],[236,0.0379],[238,0.0372],[240,0.0383],[242,0.0364],[244,0.0353],[246,0.0367],[248,0.037],[250,0.0348],[252,0.0353],[254,0.035],[256,0.0357],[258,0.0342],[260,0.0341],[262,0.0329],[266,0.0329],[268,0.0362],[270,0.04],[272,0.0446],[274,0.0468],[276,0.0432],[278,0.0338],[280,0.0322],[282,0.0316],[284,0.0345],[288,0.0408],[290,0.0408],[292,0.046],[294,0.0467],[296,0.0432],[298,0.0427],[300,0.0417],[302,0.0428],[304,0.0411],[308,0.0411],[310,0.0423],[312,0.0412],[314,0.0385],[316,0.0385],[318,0.0392],[320,0.0393],[322,0.0409],[324,0.0367],[326,0.0378],[328,0.0395],[330,0.0383],[332,0.0352],[334,0.0372],[336,0.0425],[338,0.0469],[342,0.0453],[344,0.0472],[346,0.0482],[348,0.0548],[350,0.0564],[354,0.0556],[358,0.0514],[360,0.0515],[362,0.0574],[364,0.061],[366,0.0617],[368,0.0544],[370,0.0526],[372,0.0574],[374,0.0611],[380,0.0624],[382,0.0597],[384,0.0611],[386,0.0556],[388,0.0548],[390,0.0549],[392,0.0592],[394,0.0666],[396,0.0755],[398,0.087],[400,0.0904],[402,0.0969],[404,0.0845],[406,0.0689],[408,0.0606],[410,0.0593],[412,0.0597],[414,0.0596],[416,0.0578],[418,0.0567],[420,0.0536],[422,0.0516],[424,0.0518],[426,0.0529],[428,0.0533],[430,0.0528],[432,0.0542],[434,0.0549],[436,0.0564],[438,0.0564],[440,0.0574],[442,0.0562],[444,0.0573],[446,0.0601],[448,0.0602],[450,0.0588],[452,0.06],[454,0.0619],[456,0.0633],[458,0.0608],[462,0.0639],[464,0.0667],[466,0.0727],[468,0.074],[470,0.0833],[472,0.0904],[474,0.1002],[476,0.0983],[480,0.0965],[482,0.1064],[484,0.1089],[486,0.1069],[488,0.1208],[490,0.1262],[492,0.1522],[494,0.1715],[496,0.1698],[498,0.1456],[500,0.1257],[502,0.1259],[504,0.1094],[506,0.1049],[508,0.1113],[510,0.1076],[512,0.1068],[514,0.1144],[516,0.1354],[518,0.1346],[520,0.1255],[522,0.1137],[524,0.0932],[526,0.0687],[530,0.0651],[532,0.0641],[534,0.0618],[536,0.0621],[538,0.0601],[540,0.0596],[542,0.0614],[544,0.0676],[546,0.0757],[548,0.0758],[550,0.0813],[552,0.0799],[554,0.0771],[556,0.0752],[558,0.0736],[560,0.0736],[562,0.0725],[564,0.0741],[566,0.0734],[568,0.0733],[570,0.0745],[572,0.0715],[576,0.0739],[578,0.074],[580,0.0747],[582,0.0725],[584,0.0678],[586,0.067],[588,0.065],[590,0.0683],[592,0.0821],[594,0.0845],[596,0.088],[598,0.087],[600,0.0849],[602,0.0823],[604,0.0612],[606,0.0591],[608,0.0526],[610,0.05],[612,0.0496],[614,0.0488],[616,0.0475],[618,0.0484],[620,0.049],[622,0.0526],[626,0.0541],[628,0.0535],[630,0.0506],[632,0.0488],[634,0.0494],[636,0.0454],[638,0.0461],[640,0.0487],[644,0.0483],[646,0.0503],[648,0.0521],[650,0.0513],[652,0.0548],[656,0.0596],[658,0.0581],[660,0.061],[662,0.0596],[664,0.0572],[668,0.0554],[670,0.0578],[672,0.0563],[674,0.0541],[676,0.0561],[678,0.0545],[680,0.0569],[682,0.0541],[684,0.0553],[686,0.0545],[688,0.0519],[690,0.0542],[692,0.0569],[694,0.0561],[696,0.0532],[698,0.0526],[700,0.0515],[702,0.0496],[704,0.0487],[706,0.0517],[708,0.0504],[710,0.0505],[712,0.051],[714,0.0528],[716,0.0529],[718,0.0506],[720,0.0506]]
       }
       CTRL_R_mouth_lipBiteU {
        type=MovieSceneScriptingFloatChannel
        default=0
        thinned_from=492
        keys [[-262,0.0285],[-260,0.0266],[-258,0.0234],[-256,0.0223],[-254,0.0209],[-252,0.0207],[-250,0.0208],[-248,0.02],[-244,0.02],[-242,0.0197],[-240,0.021],[-236,0.0199],[-234,0.0199],[-232,0.0195],[-230,0.0184],[-228,0.0169],[-226,0.0152],[-224,0.0142],[-220,0.0131],[-216,0.0127],[-212,0.0129],[-204,0.0119],[-202,0.0112],[-200,0.0108],[-194,0.0112],[-192,0.0109],[-190,0.0108],[-188,0.012],[-186,0.0121],[-184,0.0124],[-182,0.0135],[-180,0.0162],[-178,0.0231],[-176,0.032],[-174,0.046],[-172,0.064],[-170,0.0842],[-168,0.1032],[-166,0.1173],[-164,0.137],[-162,0.1602],[-160,0.189],[-158,0.1925],[-156,0.2244],[-154,0.2603],[-152,0.2866],[-150,0.326],[-148,0.3542],[-146,0.3649],[-144,0.3786],[-142,0.3875],[-140,0.3935],[-138,0.4012],[-136,0.399],[-134,0.3949],[-132,0.3966],[-130,0.3963],[-128,0.3928],[-126,0.3869],[-124,0.383],[-122,0.3721],[-120,0.3426],[-118,0.2889],[-116,0.1898],[-114,0.1751],[-112,0.1395],[-110,0.1131],[-108,0.0975],[-106,0.0889],[-104,0.085],[-102,0.0666],[-100,0.0596],[-98,0.059],[-96,0.054],[-94,0.0616],[-92,0.0646],[-90,0.0787],[-88,0.0814],[-86,0.091],[-84,0.1054],[-82,0.1215],[-80,0.146],[-78,0.1804],[-76,0.1913],[-74,0.2807],[-72,0.3209],[-70,0.3471],[-68,0.3548],[-66,0.3403],[-64,0.2778],[-62,0.2529],[-60,0.1288],[-58,0.1203],[-56,0.101],[-54,0.0884],[-52,0.0814],[-50,0.0761],[-48,0.0746],[-46,0.0667],[-44,0.0476],[-42,0.0382],[-40,0.052],[-38,0.0683],[-36,0.0725],[-34,0.0671],[-32,0.0586],[-30,0.0534],[-28,0.0498],[-26,0.0513],[-24,0.0494],[-22,0.0463],[-20,0.0473],[-18,0.0441],[-16,0.0399],[-14,0.04],[-12,0.0396],[-10,0.0382],[-8,0.0373],[-4,0.0367],[-2,0.0331],[0,0.0318],[4,0.032],[6,0.0326],[8,0.0329],[10,0.034],[12,0.0348],[14,0.0349],[16,0.0345],[20,0.0316],[22,0.032],[24,0.0334],[26,0.0346],[28,0.0344],[30,0.0346],[32,0.0359],[34,0.037],[36,0.0359],[38,0.0352],[40,0.0371],[42,0.0378],[44,0.0373],[46,0.037],[48,0.0366],[52,0.038],[56,0.0402],[58,0.0404],[60,0.0399],[62,0.0387],[64,0.0379],[66,0.0397],[70,0.0393],[72,0.0377],[74,0.0392],[76,0.0398],[80,0.0378],[82,0.0373],[84,0.0364],[86,0.0373],[88,0.0395],[90,0.04],[92,0.0359],[94,0.0365],[96,0.0386],[98,0.0395],[100,0.0427],[102,0.0417],[104,0.042],[106,0.0438],[108,0.0425],[110,0.0381],[112,0.0359],[114,0.035],[116,0.0313],[118,0.0279],[120,0.0275],[122,0.0274],[124,0.027],[126,0.0267],[128,0.0226],[130,0.0226],[132,0.023],[134,0.0246],[138,0.025],[140,0.025],[142,0.0243],[144,0.0232],[148,0.0259],[150,0.028],[152,0.0291],[154,0.0262],[156,0.0262],[158,0.0253],[160,0.0264],[162,0.0271],[166,0.0252],[168,0.026],[170,0.0263],[172,0.0242],[174,0.0232],[176,0.0238],[180,0.023],[182,0.0236],[184,0.0245],[186,0.0243],[188,0.0234],[190,0.0228],[192,0.0236],[194,0.024],[196,0.0234],[198,0.0233],[200,0.0229],[204,0.0228],[206,0.0211],[208,0.0216],[210,0.0216],[214,0.021],[218,0.0216],[222,0.0218],[224,0.0215],[226,0.02],[228,0.0191],[230,0.0201],[232,0.0213],[234,0.021],[240,0.0238],[242,0.0243],[244,0.0236],[246,0.0225],[248,0.0222],[250,0.0229],[252,0.0228],[256,0.0219],[258,0.022],[260,0.0206],[264,0.0187],[266,0.0183],[268,0.0208],[270,0.0235],[272,0.0241],[274,0.0253],[276,0.025],[278,0.0199],[280,0.0181],[282,0.0189],[284,0.0201],[286,0.0207],[290,0.02],[292,0.02],[294,0.0198],[296,0.02],[298,0.0208],[300,0.0212],[304,0.0208],[306,0.021],[308,0.021],[310,0.0215],[312,0.021],[314,0.0197],[316,0.0197],[318,0.0206],[322,0.0192],[324,0.0194],[332,0.0198],[334,0.02],[336,0.0212],[338,0.027],[340,0.0293],[342,0.031],[344,0.0319],[348,0.0305],[350,0.0291],[352,0.0302],[354,0.0303],[356,0.0298],[358,0.0276],[360,0.0266],[362,0.0272],[364,0.027],[366,0.0245],[368,0.0225],[370,0.0214],[372,0.0211],[374,0.02],[376,0.0198],[378,0.0188],[380,0.0186],[382,0.0187],[384,0.0198],[386,0.02],[388,0.0212],[390,0.0218],[392,0.0258],[394,0.0302],[396,0.0334],[398,0.0382],[400,0.0398],[402,0.0412],[404,0.0335],[406,0.027],[408,0.0273],[410,0.0249],[412,0.024],[414,0.0236],[416,0.0234],[418,0.0236],[422,0.0221],[424,0.0222],[426,0.0227],[428,0.0221],[430,0.0219],[432,0.0236],[434,0.0239],[436,0.0238],[444,0.026],[446,0.0271],[448,0.0272],[450,0.0266],[452,0.027],[454,0.027],[458,0.0277],[460,0.0291],[462,0.0308],[464,0.0309],[466,0.0319],[468,0.0319],[470,0.0302],[472,0.0322],[474,0.0326],[476,0.0306],[478,0.0337],[480,0.0364],[482,0.0386],[484,0.0391],[486,0.038],[488,0.0385],[490,0.0402],[492,0.0447],[494,0.0526],[496,0.0572],[498,0.0558],[500,0.0502],[502,0.0484],[504,0.0487],[506,0.0515],[508,0.0573],[510,0.0553],[512,0.0496],[514,0.0468],[516,0.0391],[518,0.0389],[520,0.0382],[522,0.0398],[526,0.0368],[528,0.0386],[530,0.0416],[532,0.0404],[534,0.0364],[536,0.03],[538,0.0265],[540,0.0262],[542,0.0315],[544,0.0312],[546,0.0295],[548,0.0282],[550,0.0281],[552,0.0275],[554,0.0266],[556,0.0266],[560,0.0254],[564,0.0257],[570,0.0255],[576,0.0258],[580,0.0252],[582,0.0235],[586,0.0179],[588,0.0171],[590,0.0201],[592,0.0261],[594,0.0272],[596,0.0332],[598,0.0329],[600,0.0254],[602,0.025],[604,0.0219],[606,0.0216],[608,0.0209],[610,0.0208],[614,0.0216],[616,0.0217],[618,0.0213],[620,0.0211],[622,0.0217],[624,0.0217],[626,0.0211],[630,0.0209],[640,0.0211],[642,0.0215],[646,0.0204],[650,0.0212],[652,0.0215],[654,0.0221],[658,0.0224],[660,0.023],[662,0.0234],[664,0.0225],[666,0.0221],[670,0.0228],[672,0.0229],[674,0.0224],[678,0.0234],[684,0.0236],[686,0.024],[688,0.0232],[690,0.0233],[692,0.0236],[694,0.0231],[696,0.0232],[700,0.0227],[702,0.0227],[704,0.0229],[706,0.0241],[708,0.024],[710,0.0233],[716,0.0231],[718,0.0226],[720,0.0226]]
       }
       CTRL_R_mouth_suckBlow {
        type=MovieSceneScriptingFloatChannel
        default=0
        thinned_from=492
        keys [[-262,0.0059],[-184,0.0059],[-182,0.006],[-178,0.0068],[-174,0.0064],[-170,0.0068],[-122,0.0069],[-120,0.0075],[-118,0.013],[-116,0.0288],[-114,0.0326],[-112,0.0409],[-110,0.0443],[-108,0.0449],[-104,0.0438],[-102,0.0389],[-100,0.0333],[-98,0.0327],[-96,0.0286],[-94,0.023],[-92,0.0211],[-90,0.0123],[-88,0.0113],[-86,0.0082],[-84,0.0068],[-66,0.0069],[-64,0.0078],[-62,0.0111],[-60,0.0321],[-58,0.0328],[-56,0.0343],[-54,0.0363],[-52,0.0368],[-50,0.0377],[-48,0.038],[-46,0.032],[-44,0.0184],[-42,0.0124],[-40,0.0118],[-38,0.0132],[-36,0.0117],[-34,0.0112],[-32,0.0117],[-26,0.0117],[-22,0.0137],[-20,0.0133],[-14,0.0145],[-12,0.0156],[-10,0.0161],[-8,0.0156],[-6,0.0157],[-4,0.0171],[-2,0.018],[0,0.0171],[2,0.0176],[4,0.019],[6,0.019],[10,0.0171],[12,0.0166],[20,0.0166],[24,0.0163],[28,0.0176],[30,0.0176],[32,0.0172],[34,0.0171],[36,0.0185],[38,0.0195],[40,0.0195],[42,0.02],[46,0.0193],[48,0.0185],[50,0.0182],[54,0.0198],[56,0.02],[60,0.0187],[64,0.019],[66,0.0186],[70,0.0187],[72,0.0195],[74,0.0195],[78,0.0203],[80,0.0193],[82,0.0193],[84,0.0211],[86,0.0212],[88,0.0199],[90,0.0195],[92,0.0199],[94,0.0181],[96,0.0143],[98,0.0132],[102,0.0095],[104,0.0107],[106,0.0146],[108,0.0146],[110,0.014],[112,0.0119],[114,0.0107],[116,0.0104],[118,0.0083],[122,0.0073],[126,0.0074],[128,0.0078],[132,0.0078],[134,0.0083],[138,0.0083],[144,0.0098],[146,0.0108],[148,0.0124],[150,0.0127],[152,0.0128],[154,0.0142],[156,0.0142],[158,0.0151],[160,0.0151],[162,0.0156],[164,0.017],[166,0.019],[168,0.0167],[170,0.0162],[172,0.019],[174,0.0175],[176,0.0157],[178,0.0152],[182,0.0152],[184,0.0166],[186,0.017],[188,0.0178],[192,0.0153],[194,0.0155],[196,0.0144],[198,0.0139],[200,0.0144],[202,0.0136],[204,0.0132],[206,0.0136],[210,0.0137],[212,0.0135],[216,0.0126],[220,0.0108],[224,0.0107],[226,0.0093],[228,0.0088],[230,0.0102],[232,0.0098],[234,0.0097],[236,0.0093],[238,0.0084],[240,0.0088],[244,0.0098],[246,0.0098],[248,0.0103],[254,0.0131],[256,0.0137],[258,0.0146],[260,0.0166],[262,0.0166],[264,0.0157],[266,0.0171],[268,0.0196],[270,0.0205],[272,0.0173],[274,0.021],[276,0.0354],[278,0.0312],[280,0.021],[282,0.0205],[284,0.0191],[286,0.0183],[288,0.0185],[290,0.0165],[292,0.0133],[294,0.0132],[296,0.0137],[300,0.0137],[302,0.0142],[304,0.0156],[308,0.0147],[310,0.0156],[314,0.0156],[316,0.0161],[318,0.0171],[320,0.0165],[322,0.0151],[326,0.0137],[328,0.0128],[330,0.0132],[332,0.0126],[336,0.0106],[338,0.009],[340,0.0085],[342,0.0076],[344,0.0073],[346,0.0077],[350,0.0068],[378,0.0068],[382,0.0059],[384,0.006],[388,0.0068],[390,0.0066],[392,0.0061],[394,0.0059],[398,0.0067],[472,0.0068],[474,0.0066],[476,0.0061],[478,0.0059],[480,0.0062],[482,0.0067],[490,0.0068],[494,0.0059],[518,0.0059],[522,0.0063],[530,0.0063],[534,0.0059],[594,0.0059],[598,0.0063],[602,0.0064],[604,0.0068],[610,0.0068],[612,0.0071],[616,0.0081],[620,0.0083],[622,0.0078],[624,0.0078],[626,0.0083],[628,0.0098],[630,0.0102],[632,0.0098],[638,0.0098],[642,0.0088],[644,0.0093],[646,0.0093],[648,0.0088],[654,0.0088],[656,0.0093],[658,0.0092],[660,0.0082],[662,0.0078],[666,0.0087],[668,0.0087],[670,0.0083],[672,0.0083],[674,0.0088],[678,0.0088],[682,0.0078],[684,0.0083],[686,0.0083],[688,0.0078],[702,0.0078],[706,0.0088],[714,0.0088],[716,0.0093],[720,0.0092]]
       }
       CTRL_L_mouth_purseU {
        type=MovieSceneScriptingFloatChannel
        default=0
        thinned_from=492
        keys [[-262,0.0712],[-260,0.0694],[-258,0.0622],[-256,0.0565],[-254,0.0565],[-252,0.0576],[-250,0.0638],[-248,0.0663],[-246,0.067],[-242,0.0674],[-240,0.0719],[-238,0.0735],[-236,0.0707],[-234,0.0696],[-232,0.0623],[-230,0.0614],[-228,0.0616],[-226,0.0613],[-224,0.0603],[-222,0.0587],[-220,0.0577],[-218,0.0562],[-216,0.0504],[-214,0.0468],[-212,0.0463],[-210,0.0445],[-208,0.0446],[-206,0.0431],[-204,0.0439],[-202,0.0487],[-200,0.0461],[-198,0.0486],[-196,0.0497],[-194,0.0495],[-192,0.0487],[-190,0.0488],[-188,0.058],[-186,0.0581],[-184,0.0549],[-180,0.0625],[-178,0.0673],[-176,0.078],[-174,0.0975],[-172,0.1124],[-170,0.127],[-168,0.144],[-166,0.1695],[-164,0.2237],[-162,0.2768],[-160,0.3255],[-158,0.3334],[-156,0.4017],[-154,0.4281],[-152,0.4567],[-150,0.4867],[-148,0.4844],[-146,0.4975],[-144,0.5031],[-142,0.4967],[-138,0.4989],[-136,0.4887],[-134,0.4913],[-132,0.498],[-130,0.4947],[-128,0.49],[-126,0.4818],[-124,0.4751],[-122,0.4578],[-120,0.4081],[-118,0.3141],[-116,0.2103],[-114,0.1869],[-112,0.1377],[-110,0.114],[-108,0.0977],[-106,0.0873],[-104,0.0824],[-102,0.0768],[-100,0.0741],[-98,0.0737],[-96,0.0726],[-94,0.0775],[-92,0.0811],[-90,0.0996],[-88,0.1087],[-86,0.1438],[-84,0.1871],[-82,0.226],[-80,0.2963],[-78,0.3726],[-76,0.3924],[-74,0.4546],[-72,0.4585],[-70,0.4677],[-68,0.4544],[-66,0.4124],[-64,0.3083],[-62,0.2791],[-60,0.1479],[-58,0.1349],[-56,0.1068],[-54,0.0922],[-52,0.0843],[-50,0.076],[-48,0.0734],[-46,0.073],[-44,0.0776],[-42,0.0804],[-40,0.0755],[-38,0.0696],[-36,0.0718],[-34,0.0621],[-32,0.0536],[-30,0.0519],[-28,0.0576],[-26,0.0563],[-24,0.0528],[-20,0.0494],[-18,0.0504],[-16,0.0491],[-14,0.0435],[-10,0.0417],[-8,0.0399],[-6,0.0398],[-4,0.0385],[-2,0.0379],[0,0.0375],[2,0.0392],[4,0.0385],[8,0.0386],[10,0.0416],[12,0.0451],[14,0.047],[16,0.0474],[18,0.0483],[20,0.0481],[22,0.0466],[24,0.0466],[26,0.0477],[28,0.0496],[30,0.0502],[32,0.05],[34,0.0517],[36,0.0512],[38,0.053],[40,0.0574],[44,0.0549],[48,0.0545],[50,0.0552],[52,0.0549],[54,0.0549],[56,0.0546],[58,0.0565],[60,0.055],[62,0.0531],[64,0.0559],[66,0.0554],[68,0.0537],[70,0.0542],[72,0.0569],[74,0.0548],[76,0.0538],[78,0.0539],[80,0.0551],[82,0.0567],[84,0.0571],[86,0.0558],[88,0.0555],[90,0.0561],[92,0.0544],[94,0.0552],[96,0.0612],[98,0.063],[100,0.0593],[102,0.0644],[104,0.0686],[106,0.0696],[108,0.0687],[110,0.0651],[112,0.0633],[114,0.0626],[116,0.06],[118,0.0567],[120,0.0561],[124,0.057],[126,0.0567],[128,0.0533],[132,0.0531],[136,0.0547],[138,0.0567],[140,0.0568],[142,0.0558],[144,0.0568],[146,0.0555],[150,0.0633],[152,0.0643],[154,0.0618],[156,0.0617],[158,0.0595],[160,0.0586],[162,0.0604],[164,0.0582],[166,0.0576],[168,0.0579],[170,0.0587],[172,0.0533],[174,0.0519],[176,0.0491],[178,0.0498],[180,0.0487],[182,0.0513],[184,0.0529],[186,0.0528],[188,0.0517],[190,0.0504],[192,0.0511],[194,0.0487],[196,0.0478],[198,0.0479],[200,0.0492],[202,0.0491],[204,0.0484],[206,0.045],[208,0.0466],[210,0.0467],[212,0.0481],[214,0.0478],[216,0.0459],[218,0.0488],[220,0.0524],[222,0.0527],[224,0.0532],[226,0.0519],[228,0.0512],[230,0.0546],[232,0.054],[234,0.0568],[236,0.0635],[238,0.0622],[240,0.0632],[242,0.0603],[244,0.0595],[246,0.0637],[248,0.0652],[250,0.0616],[252,0.0644],[254,0.0608],[256,0.0621],[258,0.0574],[260,0.0563],[262,0.0557],[264,0.0547],[266,0.0507],[268,0.0494],[270,0.0512],[272,0.0577],[274,0.0514],[276,0.0444],[278,0.0433],[280,0.0469],[282,0.0493],[284,0.0542],[286,0.0551],[288,0.0557],[290,0.0547],[292,0.0591],[294,0.0617],[296,0.0586],[298,0.0561],[300,0.0551],[302,0.0563],[304,0.0565],[306,0.0559],[308,0.0561],[310,0.0555],[314,0.0519],[316,0.0504],[318,0.0533],[320,0.0527],[322,0.0535],[324,0.0498],[326,0.0485],[328,0.0509],[330,0.0494],[332,0.0461],[334,0.0499],[336,0.0546],[338,0.0539],[340,0.0527],[342,0.0483],[344,0.0509],[346,0.053],[348,0.0582],[350,0.0588],[352,0.0577],[354,0.0576],[356,0.0563],[358,0.0527],[360,0.0529],[362,0.061],[364,0.0645],[366,0.0649],[368,0.0557],[370,0.052],[374,0.061],[376,0.0616],[378,0.0599],[380,0.0588],[382,0.0554],[384,0.0593],[386,0.0534],[388,0.0498],[390,0.0503],[392,0.0487],[394,0.051],[396,0.0576],[398,0.0695],[400,0.0736],[402,0.0821],[404,0.0578],[406,0.048],[408,0.044],[410,0.044],[412,0.0435],[414,0.0413],[416,0.0429],[418,0.0422],[420,0.0409],[422,0.0418],[424,0.0411],[428,0.0462],[430,0.0446],[436,0.0506],[438,0.0515],[440,0.0531],[442,0.0537],[446,0.0578],[448,0.0586],[450,0.0567],[452,0.0581],[454,0.0605],[456,0.0626],[458,0.0559],[460,0.0559],[462,0.0554],[464,0.0562],[466,0.056],[468,0.0551],[474,0.087],[476,0.0849],[478,0.0874],[480,0.0942],[482,0.1067],[484,0.1092],[486,0.1053],[488,0.1243],[490,0.128],[492,0.1497],[494,0.1617],[496,0.1592],[498,0.1323],[500,0.1119],[502,0.1134],[504,0.1011],[506,0.0965],[508,0.1049],[510,0.1016],[512,0.1006],[514,0.1084],[516,0.132],[518,0.1309],[520,0.119],[522,0.1042],[524,0.0832],[526,0.0592],[528,0.0574],[530,0.0563],[532,0.0549],[534,0.0531],[536,0.0526],[538,0.0477],[540,0.0467],[542,0.0478],[544,0.0566],[546,0.0661],[548,0.0643],[550,0.0707],[552,0.0696],[554,0.0675],[556,0.068],[560,0.0694],[562,0.0679],[564,0.0702],[566,0.0693],[568,0.0692],[570,0.0712],[572,0.062],[574,0.0633],[576,0.065],[578,0.0651],[580,0.0658],[582,0.0624],[584,0.0563],[586,0.0536],[588,0.0505],[590,0.0521],[592,0.0626],[594,0.0647],[596,0.0744],[598,0.0773],[600,0.0796],[602,0.0772],[604,0.0578],[606,0.0557],[608,0.0502],[610,0.0495],[612,0.0503],[614,0.0501],[616,0.0492],[618,0.0509],[620,0.0518],[622,0.0566],[624,0.0578],[626,0.0596],[628,0.059],[630,0.0568],[632,0.0533],[634,0.0537],[636,0.0468],[638,0.0493],[640,0.0529],[642,0.0522],[644,0.053],[646,0.0552],[648,0.0568],[650,0.0562],[652,0.063],[654,0.0669],[656,0.0705],[658,0.0694],[660,0.0721],[662,0.0712],[664,0.0692],[668,0.0671],[670,0.0701],[672,0.0697],[674,0.0669],[676,0.0694],[678,0.0686],[680,0.0704],[682,0.0646],[684,0.0667],[686,0.0655],[688,0.0634],[690,0.0657],[692,0.069],[694,0.0673],[696,0.0636],[698,0.0628],[700,0.0605],[702,0.0574],[704,0.0565],[706,0.0606],[708,0.0583],[710,0.0586],[712,0.0593],[714,0.0621],[716,0.0617],[718,0.0574],[720,0.0574]]
       }
       CTRL_R_mouth_towardsD {
        type=MovieSceneScriptingFloatChannel
        default=0
        thinned_from=492
        keys [[-262,0.0024],[-260,0.001],[-258,0.0008],[-256,0.0003],[-254,0],[-252,0],[-250,0.0003],[-242,0],[-240,0.0004],[-238,0.0002],[-236,0.0007],[-234,0.0007],[-232,0.0011],[-228,0.0011],[-226,0.0008],[-222,0.0008],[-220,0.0004],[-218,0.0004],[-216,0.001],[-214,0.0013],[-208,0.0013],[-206,0.0008],[-204,0.0011],[-202,0.001],[-200,0.0007],[-196,0.0012],[-186,0.001],[-182,0.0008],[-180,0.0012],[-176,0.0014],[-168,0.0027],[-162,0.0027],[-154,0.0029],[-152,0.0032],[-150,0.0032],[-148,0.0029],[-146,0.0041],[-144,0.0043],[-142,0.0036],[-140,0.004],[-138,0.0046],[-136,0.0042],[-134,0.0052],[-132,0.005],[-130,0.005],[-128,0.0046],[-126,0.0044],[-124,0.0047],[-122,0.0046],[-118,0.0039],[-116,0.0034],[-98,0.0029],[-82,0.0025],[-78,0.003],[-76,0.0031],[-74,0.0034],[-72,0.0033],[-70,0.0035],[-68,0.0033],[-64,0.0034],[-62,0.0033],[-60,0.0027],[-56,0.0024],[-42,0.0025],[-38,0.0029],[-34,0.0028],[-32,0.0029],[-30,0.0028],[-20,0.0031],[-18,0.0029],[-4,0.0033],[-2,0.0032],[0,0.0035],[6,0.0032],[12,0.0034],[20,0.0033],[26,0.0042],[28,0.0041],[32,0.0034],[34,0.0036],[36,0.0041],[38,0.0043],[42,0.0042],[44,0.0037],[48,0.0038],[50,0.0035],[54,0.0039],[58,0.0034],[60,0.0036],[66,0.0031],[68,0.0033],[70,0.0032],[72,0.0029],[80,0.0034],[82,0.0032],[84,0.0036],[86,0.0037],[88,0.0034],[90,0.0033],[92,0.0037],[94,0.0032],[100,0.0027],[104,0.0028],[106,0.0027],[108,0.0029],[110,0.0028],[112,0.0024],[116,0.0008],[124,0.0015],[130,0.0013],[132,0.0019],[136,0.0015],[138,0.0018],[146,0.0012],[148,0.0013],[150,0.0018],[152,0.0015],[156,0.0016],[162,0.0012],[164,0.0009],[166,0.0012],[168,0.0008],[172,0.0009],[174,0.0013],[176,0.0009],[192,0.0007],[194,0.0005],[196,0.0007],[198,0.0014],[200,0.0017],[204,0.0014],[206,0.0007],[210,0.0009],[212,0.0007],[214,0.0008],[216,0.0006],[218,0.001],[222,0.0012],[228,0.0004],[230,0.0011],[232,0.001],[234,0.0005],[236,0.0008],[238,0.0006],[242,0.001],[244,0.0014],[248,0.0013],[250,0.0007],[252,0.002],[254,0.0025],[256,0.0018],[258,0.003],[260,0.0045],[262,0.0044],[264,0.0037],[266,0.0041],[268,0.0053],[270,0.0047],[272,0.0036],[274,0.0032],[276,0.0051],[278,0.0045],[280,0.0055],[286,0.0044],[290,0.0027],[292,0.0022],[294,0.0023],[298,0.0017],[300,0.002],[302,0.0028],[304,0.0029],[306,0.0032],[308,0.0029],[310,0.0031],[312,0.0027],[314,0.0031],[316,0.003],[318,0.0036],[320,0.003],[322,0.003],[324,0.0039],[328,0.003],[332,0.0025],[334,0.0025],[336,0.0016],[338,0.0023],[340,0.0024],[342,0.0028],[344,0.0024],[346,0.0015],[348,0.0009],[350,0.0012],[356,0.0015],[358,0.0007],[360,0.0006],[364,0.0006],[368,0.0002],[370,0.0004],[372,0.0004],[374,0.0007],[376,0.0008],[380,0.0001],[384,0.0003],[386,0.0016],[388,0.0025],[390,0.0025],[392,0.002],[394,0.0014],[396,0.0019],[398,0.0023],[402,0.0021],[404,0.0028],[406,0.0024],[410,0.0021],[412,0.0023],[416,0.0024],[420,0.0022],[424,0.0024],[428,0.0018],[432,0.0019],[434,0.0016],[438,0.0017],[440,0.002],[444,0.002],[446,0.0021],[448,0.002],[454,0.0022],[458,0.0016],[464,0.0017],[466,0.002],[468,0.0018],[470,0.0007],[472,0.002],[474,0.0024],[476,0.0024],[478,0.0011],[480,0.001],[482,0.0013],[484,0.0013],[488,0.0004],[490,0.0007],[492,0.0003],[494,0.0007],[496,0.0003],[498,0.0004],[500,0.0001],[502,0.0005],[504,0.0001],[506,0],[508,0.0005],[510,0.0005],[514,0],[518,0],[520,0.001],[522,0.0016],[524,0.0013],[526,0.0023],[530,0.0027],[536,0.0029],[538,0.0022],[544,0.0019],[548,0.0006],[554,0.0004],[558,0.0002],[562,0.0004],[564,0.0002],[566,0.0005],[570,0.0006],[572,0.0001],[576,0.0004],[582,0.0001],[584,0.0002],[586,0.0005],[588,0.0006],[592,0.0011],[594,0.0011],[596,0.0009],[598,0.0014],[600,0.0017],[612,0.002],[614,0.0022],[620,0.0019],[622,0.0017],[626,0.0015],[628,0.0016],[638,0.0015],[640,0.0018],[642,0.0017],[646,0.0011],[650,0.0018],[652,0.0017],[660,0.0026],[664,0.0026],[668,0.0023],[672,0.0026],[674,0.0024],[682,0.0021],[686,0.0022],[688,0.002],[690,0.0021],[700,0.0016],[702,0.002],[708,0.0015],[712,0.0018],[716,0.0017],[720,0.0015]]
       }
       CTRL_L_mouth_lipBiteD {
        type=MovieSceneScriptingFloatChannel
        default=0
        thinned_from=492
        keys [[-262,0.0018],[-258,0.0006],[-254,0],[-248,0.0002],[-246,0],[-242,0],[-240,0.0005],[-236,0.0002],[-234,0.0003],[-232,0.0006],[-228,0.0007],[-226,0.0005],[-224,0.0006],[-222,0.0008],[-218,0],[-214,0.0009],[-208,0.0009],[-206,0.0007],[-202,0.0009],[-200,0.0004],[-198,0.0002],[-194,0.0009],[-190,0.001],[-186,0.0008],[-184,0.0003],[-182,0.0002],[-178,0.0012],[-176,0.0013],[-172,0.0019],[-168,0.0021],[-166,0.0028],[-164,0.0042],[-162,0.0078],[-160,0.029],[-158,0.0344],[-156,0.0831],[-154,0.1584],[-152,0.2111],[-150,0.2873],[-148,0.3521],[-144,0.4023],[-142,0.4118],[-140,0.4198],[-138,0.4333],[-134,0.4314],[-132,0.4264],[-130,0.4167],[-128,0.4125],[-124,0.4164],[-122,0.406],[-120,0.385],[-118,0.3411],[-116,0.2601],[-114,0.2457],[-112,0.2093],[-110,0.1807],[-108,0.1628],[-106,0.1507],[-104,0.1453],[-102,0.114],[-100,0.0962],[-98,0.0945],[-96,0.0683],[-92,0.0692],[-90,0.0645],[-88,0.0626],[-86,0.0566],[-84,0.0572],[-82,0.0641],[-80,0.0777],[-78,0.1069],[-76,0.1182],[-74,0.2624],[-72,0.3313],[-70,0.3636],[-68,0.3657],[-66,0.3556],[-64,0.3031],[-62,0.2783],[-60,0.1508],[-58,0.1427],[-56,0.126],[-54,0.12],[-52,0.1182],[-50,0.1196],[-48,0.1206],[-46,0.1251],[-44,0.1103],[-42,0.0919],[-40,0.0894],[-38,0.0872],[-36,0.0772],[-34,0.0681],[-32,0.0575],[-30,0.0489],[-28,0.0448],[-26,0.0446],[-24,0.0424],[-22,0.0404],[-20,0.0412],[-18,0.0391],[-16,0.0352],[-14,0.0361],[-12,0.0383],[-10,0.0393],[-8,0.0384],[-6,0.0369],[-4,0.0381],[-2,0.0354],[0,0.0319],[2,0.0325],[4,0.0328],[6,0.0329],[8,0.0323],[10,0.0296],[12,0.0253],[16,0.0246],[20,0.0218],[22,0.0221],[24,0.0238],[26,0.0258],[30,0.0247],[32,0.0255],[38,0.0261],[40,0.0268],[42,0.0272],[46,0.027],[48,0.0271],[50,0.0286],[54,0.032],[56,0.0325],[58,0.032],[60,0.0312],[62,0.0308],[64,0.0312],[66,0.0321],[68,0.0306],[72,0.0307],[74,0.0332],[76,0.0326],[78,0.0318],[80,0.0325],[82,0.0313],[84,0.0305],[86,0.0331],[88,0.0353],[90,0.0356],[92,0.0321],[94,0.0337],[96,0.0412],[100,0.0467],[102,0.0469],[104,0.0482],[106,0.0524],[108,0.0494],[110,0.0412],[112,0.0364],[114,0.0343],[118,0.0249],[120,0.0242],[122,0.0244],[126,0.0239],[128,0.0188],[130,0.0188],[132,0.0198],[134,0.0222],[136,0.0222],[142,0.0229],[146,0.023],[148,0.0241],[150,0.0265],[152,0.0276],[154,0.0273],[156,0.0275],[158,0.0295],[160,0.0301],[162,0.0295],[164,0.0301],[166,0.0324],[168,0.0337],[170,0.0331],[172,0.0306],[174,0.0298],[176,0.0305],[182,0.0308],[184,0.0328],[186,0.0328],[188,0.0318],[190,0.0281],[192,0.0276],[194,0.028],[196,0.027],[198,0.0269],[200,0.0256],[202,0.0246],[206,0.0243],[208,0.0241],[210,0.024],[212,0.0226],[214,0.0228],[216,0.0229],[220,0.0209],[224,0.0178],[226,0.0158],[228,0.0146],[230,0.0142],[232,0.0158],[234,0.0155],[236,0.0161],[238,0.0181],[240,0.0196],[242,0.0193],[244,0.0175],[246,0.0173],[248,0.017],[250,0.0172],[252,0.0187],[254,0.0191],[258,0.0176],[260,0.0188],[262,0.0196],[264,0.0182],[266,0.0175],[268,0.0171],[272,0.0152],[276,0.015],[278,0.0141],[280,0.0179],[282,0.02],[284,0.0206],[286,0.019],[288,0.0169],[290,0.017],[292,0.0167],[294,0.0162],[296,0.0172],[298,0.0164],[300,0.0148],[304,0.0172],[306,0.019],[308,0.0194],[310,0.0182],[312,0.0177],[314,0.0163],[316,0.0159],[318,0.0182],[320,0.0187],[322,0.0175],[324,0.0166],[326,0.0154],[328,0.0146],[332,0.0158],[334,0.0149],[336,0.013],[340,0.0099],[342,0.0086],[346,0.0092],[348,0.0085],[350,0.0071],[352,0.0077],[356,0.0079],[358,0.006],[360,0.0048],[364,0.0047],[368,0.0033],[372,0.0022],[374,0.0011],[380,0.0005],[384,0.0008],[386,0.0017],[388,0.0037],[390,0.0037],[392,0.0028],[394,0.0016],[396,0.0018],[398,0.0034],[400,0.0039],[402,0.0053],[404,0.0108],[406,0.0116],[408,0.0072],[410,0.0071],[414,0.006],[416,0.0057],[418,0.0057],[422,0.0046],[424,0.0044],[428,0.0045],[430,0.0043],[432,0.0049],[436,0.004],[438,0.0041],[442,0.0056],[444,0.0061],[446,0.0071],[448,0.0075],[450,0.0065],[452,0.0063],[454,0.0068],[456,0.0063],[458,0.0062],[462,0.0076],[464,0.0076],[466,0.0086],[468,0.0082],[470,0.0043],[472,0.0025],[476,0.0026],[478,0.0015],[480,0.0012],[482,0.0015],[484,0.0016],[486,0.0008],[488,0.0004],[490,0.0006],[494,0.0002],[500,0.0001],[504,0.0002],[506,0],[510,0.0006],[514,0],[518,0],[522,0.0017],[524,0.0018],[526,0.0026],[530,0.0029],[536,0.0027],[538,0.0022],[540,0.0021],[542,0.0017],[544,0.0016],[546,0.0012],[548,0.0004],[552,0],[554,0],[556,0.0002],[560,0.0002],[562,0.0004],[568,0.0004],[570,0.0006],[572,0],[578,0.0005],[584,0],[590,0.0006],[592,0.001],[594,0.0011],[596,0.0009],[598,0.0012],[600,0.0019],[604,0.0018],[606,0.0023],[610,0.0061],[612,0.0074],[614,0.0083],[616,0.0085],[620,0.0083],[622,0.009],[624,0.0091],[626,0.0095],[628,0.0094],[630,0.0099],[632,0.0101],[638,0.009],[640,0.009],[642,0.0088],[644,0.0082],[646,0.0078],[654,0.01],[658,0.01],[662,0.0095],[664,0.009],[668,0.0086],[670,0.0088],[672,0.0093],[674,0.0092],[676,0.0085],[678,0.0085],[680,0.0088],[682,0.0081],[684,0.0079],[686,0.0081],[688,0.0078],[690,0.0082],[692,0.0082],[694,0.0079],[696,0.0083],[698,0.0081],[700,0.0072],[702,0.0067],[706,0.0071],[708,0.0068],[710,0.006],[712,0.0059],[716,0.0068],[718,0.0062],[720,0.0062]]
       }
       CTRL_R_mouth_purseU {
        type=MovieSceneScriptingFloatChannel
        default=0
        thinned_from=492
        keys [[-262,0.1344],[-260,0.1317],[-258,0.1237],[-256,0.1166],[-254,0.1161],[-252,0.1165],[-250,0.122],[-246,0.1216],[-244,0.1218],[-242,0.1224],[-240,0.1271],[-238,0.1288],[-236,0.1256],[-234,0.1242],[-232,0.1151],[-230,0.1145],[-228,0.1157],[-226,0.1157],[-224,0.1141],[-222,0.1119],[-220,0.1104],[-218,0.1094],[-216,0.1035],[-214,0.1],[-212,0.0995],[-210,0.097],[-208,0.0996],[-206,0.0973],[-204,0.099],[-202,0.103],[-200,0.0989],[-198,0.1018],[-196,0.1026],[-194,0.1024],[-192,0.1016],[-190,0.1016],[-188,0.1113],[-186,0.1115],[-184,0.108],[-182,0.1122],[-178,0.1189],[-176,0.1274],[-174,0.1426],[-172,0.1523],[-170,0.1627],[-168,0.1766],[-166,0.2],[-164,0.2506],[-162,0.2995],[-160,0.3444],[-158,0.3513],[-156,0.4111],[-154,0.4389],[-152,0.4657],[-150,0.497],[-148,0.5012],[-144,0.5269],[-142,0.5207],[-138,0.5266],[-136,0.5182],[-134,0.5168],[-132,0.5234],[-130,0.522],[-128,0.5182],[-124,0.506],[-122,0.4914],[-120,0.451],[-118,0.3742],[-116,0.2819],[-114,0.2576],[-112,0.2055],[-110,0.1784],[-108,0.1609],[-106,0.1499],[-104,0.1445],[-102,0.1371],[-100,0.1328],[-98,0.1323],[-96,0.1322],[-94,0.1384],[-92,0.1425],[-90,0.1604],[-88,0.1681],[-86,0.1977],[-84,0.2346],[-82,0.2677],[-80,0.3265],[-78,0.39],[-76,0.4068],[-74,0.4652],[-70,0.4854],[-68,0.4751],[-66,0.4462],[-64,0.3588],[-62,0.333],[-60,0.2132],[-58,0.1999],[-56,0.1698],[-54,0.1516],[-52,0.1426],[-50,0.1353],[-48,0.1332],[-46,0.1384],[-44,0.1381],[-42,0.1407],[-40,0.1383],[-38,0.138],[-36,0.1392],[-32,0.1207],[-30,0.117],[-28,0.1216],[-26,0.1192],[-24,0.1146],[-20,0.1124],[-18,0.1108],[-16,0.1088],[-14,0.1026],[-8,0.0989],[-6,0.0996],[-4,0.0966],[-2,0.0965],[0,0.0946],[2,0.0961],[4,0.0962],[6,0.098],[8,0.0985],[10,0.1006],[12,0.1052],[14,0.1091],[16,0.1093],[20,0.1113],[22,0.1086],[24,0.1086],[26,0.1097],[28,0.113],[30,0.1144],[32,0.1139],[34,0.1163],[36,0.1145],[38,0.1161],[40,0.1198],[42,0.1171],[44,0.1146],[46,0.1148],[48,0.1158],[54,0.1177],[56,0.1172],[58,0.1196],[62,0.1158],[64,0.1189],[66,0.1192],[68,0.1173],[70,0.1177],[72,0.1199],[76,0.1156],[78,0.1157],[80,0.1184],[82,0.1204],[84,0.1198],[86,0.1179],[88,0.1174],[90,0.1178],[92,0.1162],[94,0.1174],[96,0.1238],[98,0.1261],[100,0.1226],[102,0.1286],[104,0.1321],[106,0.1317],[108,0.1297],[110,0.1257],[112,0.125],[114,0.1249],[116,0.123],[118,0.1197],[120,0.119],[122,0.1204],[124,0.1189],[126,0.1188],[128,0.116],[130,0.116],[132,0.1155],[134,0.1161],[136,0.1174],[138,0.1181],[140,0.1173],[142,0.1162],[144,0.1156],[146,0.1136],[148,0.1172],[150,0.1216],[152,0.1198],[154,0.1143],[156,0.1141],[158,0.1117],[160,0.1101],[162,0.1109],[164,0.1076],[166,0.1067],[168,0.1088],[170,0.1083],[172,0.1017],[174,0.1009],[176,0.0972],[178,0.0968],[180,0.0957],[182,0.0993],[184,0.1008],[186,0.1004],[190,0.098],[192,0.0975],[194,0.0953],[198,0.0948],[202,0.0954],[204,0.0949],[206,0.0919],[208,0.0923],[210,0.0923],[212,0.0936],[214,0.0934],[216,0.0914],[218,0.0947],[220,0.0976],[222,0.1003],[226,0.1007],[228,0.0997],[230,0.1039],[232,0.1051],[234,0.1097],[236,0.1205],[238,0.1207],[240,0.1224],[242,0.1189],[244,0.1176],[246,0.1229],[248,0.1225],[250,0.1185],[252,0.1216],[254,0.1188],[256,0.1204],[258,0.1154],[260,0.112],[262,0.1113],[264,0.111],[266,0.1068],[268,0.1095],[270,0.1162],[272,0.1224],[274,0.119],[276,0.1048],[278,0.0946],[280,0.0904],[282,0.0904],[286,0.0983],[288,0.101],[290,0.1001],[292,0.1027],[294,0.1063],[298,0.1012],[300,0.1004],[302,0.1],[304,0.0989],[306,0.0988],[308,0.1],[310,0.1023],[312,0.0994],[314,0.0953],[316,0.0935],[318,0.0974],[320,0.096],[322,0.0965],[324,0.0912],[326,0.092],[328,0.0938],[330,0.0926],[332,0.0893],[334,0.095],[336,0.1028],[338,0.1068],[340,0.108],[342,0.1064],[344,0.1109],[346,0.1163],[348,0.1221],[350,0.1233],[352,0.1235],[354,0.1225],[358,0.1139],[360,0.1132],[362,0.1208],[364,0.1251],[366,0.125],[368,0.1164],[370,0.114],[372,0.1195],[374,0.1254],[376,0.1262],[380,0.1272],[382,0.1261],[384,0.134],[386,0.1312],[388,0.1243],[390,0.1241],[392,0.1208],[394,0.1214],[396,0.1278],[398,0.1362],[400,0.1387],[402,0.1399],[404,0.1255],[406,0.1166],[408,0.1118],[410,0.1134],[412,0.1123],[414,0.1101],[418,0.1099],[420,0.1076],[422,0.109],[424,0.1089],[426,0.1116],[428,0.1133],[430,0.112],[432,0.1135],[436,0.1178],[440,0.1184],[442,0.1183],[446,0.1213],[448,0.122],[450,0.1209],[452,0.122],[456,0.1258],[458,0.1207],[462,0.12],[466,0.1217],[468,0.1206],[470,0.1292],[472,0.1422],[474,0.1566],[476,0.1519],[478,0.1488],[480,0.1536],[482,0.1669],[484,0.1696],[486,0.1682],[488,0.1803],[490,0.1828],[492,0.2078],[494,0.2261],[496,0.2253],[498,0.2008],[500,0.1798],[502,0.1805],[504,0.1635],[506,0.1568],[508,0.1655],[510,0.1637],[512,0.1623],[514,0.1692],[516,0.1949],[518,0.1932],[520,0.1756],[522,0.1588],[524,0.1366],[526,0.1142],[528,0.1113],[530,0.1104],[534,0.1072],[536,0.1107],[538,0.1057],[540,0.1044],[542,0.1068],[544,0.115],[546,0.1239],[548,0.1206],[550,0.1267],[552,0.1256],[554,0.1223],[556,0.1221],[558,0.1225],[560,0.1239],[562,0.1216],[564,0.1235],[566,0.1231],[568,0.1231],[570,0.1245],[572,0.116],[576,0.1184],[578,0.118],[580,0.1186],[582,0.1159],[584,0.1118],[586,0.1126],[588,0.1087],[590,0.109],[592,0.1244],[594,0.1274],[596,0.1376],[598,0.1397],[600,0.1423],[602,0.1395],[604,0.1165],[606,0.1145],[608,0.1088],[610,0.1073],[612,0.1061],[616,0.1028],[618,0.1042],[620,0.1051],[622,0.1076],[624,0.1084],[626,0.1104],[628,0.1109],[630,0.1085],[632,0.1045],[634,0.104],[636,0.0966],[638,0.0995],[640,0.103],[642,0.102],[644,0.1026],[646,0.1051],[648,0.1064],[650,0.1044],[652,0.1092],[654,0.1124],[656,0.1151],[658,0.1141],[660,0.116],[662,0.1151],[664,0.1127],[668,0.1102],[670,0.1124],[672,0.1122],[674,0.1099],[676,0.112],[678,0.1117],[680,0.1136],[682,0.1079],[684,0.1092],[686,0.1083],[688,0.1064],[690,0.1093],[692,0.1125],[696,0.1081],[698,0.1069],[702,0.1028],[704,0.1017],[706,0.1051],[708,0.1027],[710,0.1028],[712,0.1035],[714,0.1061],[716,0.1047],[718,0.1011],[720,0.1011]]
       }
       CTRL_R_mouth_cornerDepress {
        type=MovieSceneScriptingFloatChannel
        default=0
        thinned_from=492
        keys [[-262,0.0474],[-260,0.0551],[-258,0.0572],[-254,0.0546],[-252,0.054],[-250,0.0507],[-248,0.0455],[-246,0.0445],[-244,0.0444],[-242,0.0436],[-240,0.0418],[-238,0.041],[-236,0.0406],[-234,0.0406],[-232,0.0401],[-230,0.0399],[-228,0.0416],[-226,0.0436],[-224,0.0459],[-222,0.0479],[-220,0.0485],[-218,0.0493],[-216,0.0543],[-214,0.0603],[-212,0.0612],[-210,0.0627],[-208,0.0651],[-206,0.0661],[-204,0.0697],[-202,0.0724],[-200,0.0758],[-196,0.084],[-194,0.0896],[-192,0.0928],[-190,0.0936],[-188,0.0991],[-186,0.0991],[-184,0.096],[-182,0.09],[-180,0.0816],[-178,0.0682],[-176,0.0529],[-174,0.0386],[-172,0.0294],[-170,0.0225],[-168,0.016],[-166,0.0124],[-164,0.0104],[-162,0.0095],[-158,0.0097],[-156,0.0115],[-154,0.0192],[-152,0.0272],[-150,0.0398],[-148,0.0521],[-146,0.0609],[-144,0.0669],[-142,0.0689],[-140,0.0697],[-138,0.0711],[-136,0.0732],[-134,0.0705],[-130,0.0669],[-128,0.0672],[-126,0.0628],[-124,0.0581],[-122,0.0537],[-120,0.0511],[-118,0.0494],[-116,0.045],[-114,0.043],[-112,0.0385],[-108,0.0311],[-106,0.0287],[-104,0.0276],[-102,0.0256],[-100,0.0242],[-98,0.0241],[-96,0.0171],[-94,0.0144],[-92,0.0145],[-88,0.0138],[-86,0.0128],[-84,0.0125],[-82,0.0131],[-80,0.0126],[-78,0.0115],[-76,0.0113],[-74,0.0193],[-72,0.0261],[-70,0.0325],[-68,0.0358],[-66,0.0369],[-64,0.0339],[-62,0.0325],[-60,0.0243],[-58,0.0233],[-54,0.0187],[-52,0.0162],[-50,0.0139],[-48,0.0133],[-46,0.0136],[-44,0.0134],[-42,0.0139],[-40,0.0139],[-38,0.0196],[-36,0.0246],[-34,0.0284],[-32,0.0271],[-30,0.0252],[-28,0.0226],[-26,0.0213],[-24,0.0099],[268,0.0097],[270,0.0138],[272,0.0153],[274,0.0175],[278,0.0124],[280,0.0081],[284,0.0056],[288,0.0026],[290,0.0031],[294,0.003],[296,0.0017],[300,0.0028],[302,0.0038],[306,0.0038],[308,0.0033],[312,0.0029],[314,0.0035],[316,0.0038],[318,0.0046],[320,0.0046],[322,0.004],[328,0.0032],[330,0.0024],[334,0.002],[336,0.0083],[348,0.0088],[352,0.0095],[354,0.0096],[356,0.0109],[360,0.0125],[362,0.0148],[364,0.0166],[366,0.0172],[370,0.017],[372,0.0183],[374,0.019],[376,0.0191],[378,0.0204],[382,0.0224],[384,0.0261],[386,0.0308],[388,0.036],[390,0.0358],[392,0.0411],[394,0.0484],[396,0.0559],[398,0.0564],[400,0.0554],[402,0.0483],[404,0.0594],[406,0.0625],[408,0.0679],[410,0.0513],[412,0.0511],[416,0.0526],[418,0.053],[420,0.0497],[422,0.0462],[424,0.0453],[426,0.0456],[428,0.0469],[430,0.0464],[432,0.0454],[434,0.0458],[436,0.0447],[438,0.044],[440,0.042],[442,0.0409],[444,0.0405],[446,0.0408],[448,0.0427],[450,0.0433],[452,0.0427],[454,0.0414],[456,0.0407],[458,0.0415],[460,0.0401],[464,0.0359],[466,0.0369],[468,0.0389],[470,0.0386],[472,0.0353],[474,0.0329],[476,0.0326],[478,0.0331],[480,0.0338],[482,0.0349],[484,0.0351],[486,0.0377],[488,0.0379],[490,0.0391],[492,0.0453],[494,0.0533],[496,0.0585],[498,0.0575],[500,0.0527],[502,0.0481],[504,0.0403],[506,0.0316],[508,0.0236],[510,0.0186],[512,0.0187],[514,0.0177],[516,0.0214],[518,0.0221],[520,0.0277],[522,0.0286],[524,0.0279],[526,0.0264],[528,0.0265],[530,0.0272],[534,0.0308],[536,0.0335],[538,0.0379],[540,0.0386],[542,0.0432],[544,0.0439],[546,0.0434],[548,0.0413],[550,0.0401],[552,0.0385],[554,0.0374],[556,0.0367],[558,0.0374],[560,0.0385],[562,0.0376],[564,0.0369],[568,0.0371],[570,0.0387],[572,0.0389],[574,0.0404],[576,0.0428],[578,0.0477],[582,0.0514],[584,0.0523],[588,0.0519],[590,0.0532],[592,0.0556],[594,0.0558],[596,0.0532],[598,0.0481],[600,0.0379],[602,0.0368],[604,0.0278],[606,0.0263],[608,0.0217],[610,0.0193],[612,0.0174],[614,0.0167],[618,0.0145],[620,0.0142],[622,0.0125],[624,0.0123],[626,0.0114],[628,0.0102],[632,0.0102],[638,0.0092],[640,0.0097],[642,0.0099],[646,0.0095],[648,0.009],[650,0.0091],[656,0.0086],[658,0.0087],[660,0.0091],[662,0.0093],[666,0.0087],[668,0.0081],[674,0.0073],[678,0.0075],[680,0.008],[684,0.0077],[686,0.0079],[688,0.0084],[690,0.0085],[694,0.0083],[698,0.0085],[702,0.008],[704,0.0079],[706,0.0081],[712,0.0078],[714,0.0081],[716,0.0081],[718,0.0083],[720,0.0083]]
       }
       CTRL_R_mouth_upperLipRaise {
        type=MovieSceneScriptingFloatChannel
        default=0
        thinned_from=492
        keys [[-262,0.0019],[-260,0.0007],[-258,0.0005],[-254,0],[-238,0],[-232,0.0004],[-228,0.0003],[-226,0.0001],[-222,0.0002],[-220,0],[-218,0],[-214,0.0004],[-208,0.0006],[-206,0.0002],[-204,0.0002],[-202,0],[-198,0],[-190,0.0004],[-186,0.0003],[-184,0],[-182,0.0001],[-178,0.0008],[-176,0.001],[-166,0.0023],[-162,0.0022],[-158,0.0023],[-154,0.002],[-152,0.0021],[-148,0.0016],[-146,0.0022],[-142,0.0022],[-140,0.0024],[-138,0.0021],[-134,0.0023],[-130,0.0018],[-122,0.0018],[-118,0.0023],[-108,0.0021],[-102,0.0024],[-92,0.0022],[-86,0.0024],[-82,0.0021],[-76,0.0024],[-68,0.0018],[-64,0.0023],[-56,0.0019],[-54,0.0021],[-52,0.0019],[-48,0.0021],[-46,0.0017],[-44,0.0019],[-42,0.0016],[-40,0.0022],[-36,0.0026],[-30,0.0023],[-12,0.0028],[-8,0.0027],[-6,0.0029],[4,0.0027],[8,0.0025],[10,0.0026],[24,0.0022],[36,0.0022],[38,0.0024],[40,0.0023],[42,0.0024],[44,0.0021],[58,0.0024],[60,0.0025],[64,0.0024],[66,0.0021],[84,0.0021],[94,0.0022],[96,0.0023],[104,0.0024],[106,0.0021],[108,0.0022],[112,0.0018],[116,0.0005],[120,0.0009],[122,0.0008],[124,0.0012],[130,0.0013],[132,0.0018],[134,0.0011],[136,0.001],[138,0.0015],[148,0.001],[150,0.0015],[152,0.0012],[156,0.0013],[158,0.0009],[160,0.0008],[162,0.0009],[164,0.0006],[166,0.0008],[168,0.0006],[172,0.0009],[174,0.0013],[176,0.0006],[188,0.0001],[196,0.0002],[200,0.0011],[204,0.0007],[206,0.0002],[210,0.0004],[212,0.0002],[214,0.0004],[216,0.0003],[222,0.0006],[228,0.0001],[230,0.0005],[232,0.0005],[234,0.0002],[236,0.0002],[238,0],[242,0.0004],[244,0.0003],[246,0],[264,0],[266,0.0003],[268,0.0009],[272,0.0012],[274,0.0017],[276,0.0037],[278,0.0015],[280,0.0032],[282,0.0028],[284,0.003],[286,0.0028],[288,0.0018],[292,0.0007],[294,0.001],[298,0.0006],[300,0.0009],[302,0.0014],[304,0.0011],[306,0.0014],[308,0.0013],[310,0.0014],[312,0.0008],[314,0.0008],[318,0.0015],[320,0.001],[324,0.0017],[326,0.0014],[328,0.0008],[330,0.001],[332,0.001],[334,0.0013],[336,0.0012],[344,0.0013],[350,0.0011],[356,0.0012],[358,0.0005],[362,0.0007],[364,0.0006],[368,0],[370,0.0005],[378,0.0003],[380,0.0001],[382,0.0003],[384,0.0003],[388,0.002],[390,0.002],[394,0.0012],[396,0.0019],[398,0.0023],[400,0.0023],[402,0.0018],[404,0.0028],[406,0.0023],[410,0.002],[416,0.0021],[420,0.002],[422,0.0021],[424,0.0021],[428,0.0015],[430,0.0018],[432,0.0017],[434,0.0014],[438,0.0015],[440,0.0019],[444,0.0016],[446,0.0017],[448,0.0015],[454,0.0019],[458,0.0016],[462,0.0017],[464,0.0016],[466,0.0018],[468,0.0016],[470,0.0005],[472,0.0019],[474,0.0021],[476,0.002],[478,0.001],[480,0.001],[482,0.0013],[484,0.0014],[486,0.0008],[490,0.0009],[492,0.0004],[494,0.0005],[496,0.0002],[498,0.0002],[500,0],[502,0.0003],[504,0],[510,0.0004],[514,0],[518,0],[520,0.001],[522,0.0016],[524,0.0013],[526,0.0021],[530,0.0024],[536,0.002],[538,0.0016],[540,0.0015],[542,0.0016],[544,0.0012],[548,0.0002],[588,0],[590,0.0004],[594,0.0007],[596,0.0009],[600,0.0018],[602,0.0018],[604,0.0016],[614,0.0022],[618,0.0018],[626,0.0013],[632,0.0015],[634,0.0012],[636,0.0011],[638,0.0015],[640,0.0016],[642,0.0015],[644,0.001],[646,0.001],[648,0.0015],[652,0.0017],[654,0.0014],[658,0.0017],[660,0.0016],[664,0.0019],[666,0.0016],[668,0.0016],[670,0.0018],[674,0.0016],[676,0.0017],[680,0.0013],[686,0.0016],[696,0.0015],[700,0.001],[702,0.0013],[706,0.0015],[716,0.0015],[720,0.0013]]
       }
       CTRL_R_mouth_dimple {
        type=MovieSceneScriptingFloatChannel
        default=0.7036
        thinned_from=492
        keys [[-262,0.0016],[-258,0.0013],[-254,0.0004],[-250,0.0001],[-232,0.0005],[-228,0.0008],[-216,0.0004],[-208,0.0009],[-204,0.0008],[-202,0.001],[-196,0.0005],[-186,0.0008],[-182,0.0005],[-180,0.0005],[-176,0.0011],[-172,0.002],[-170,0.0029],[-168,0.0055],[-166,0.0115],[-164,0.0222],[-162,0.0331],[-160,0.0472],[-158,0.0485],[-156,0.0597],[-154,0.072],[-152,0.0837],[-150,0.0929],[-146,0.1147],[-144,0.1187],[-142,0.1274],[-140,0.1281],[-138,0.1262],[-136,0.127],[-134,0.1232],[-132,0.1232],[-130,0.1247],[-128,0.1281],[-126,0.1302],[-124,0.1332],[-122,0.1345],[-120,0.1308],[-118,0.1213],[-116,0.0978],[-114,0.0914],[-112,0.0759],[-110,0.064],[-108,0.0555],[-106,0.0512],[-104,0.0496],[-102,0.0414],[-100,0.0363],[-98,0.0358],[-96,0.0276],[-92,0.0299],[-90,0.0379],[-88,0.0401],[-86,0.0475],[-84,0.0538],[-82,0.0595],[-80,0.0708],[-78,0.0841],[-76,0.0879],[-74,0.1179],[-72,0.134],[-70,0.151],[-68,0.1584],[-66,0.1574],[-64,0.1458],[-62,0.1402],[-60,0.1034],[-58,0.0957],[-56,0.0769],[-54,0.0649],[-52,0.061],[-50,0.0605],[-48,0.0604],[-44,0.0476],[-40,0.03],[-38,0.0238],[-36,0.0178],[-34,0.0138],[-32,0.0105],[-30,0.01],[-28,0.009],[-26,0.0095],[-24,0.0598],[104,0.0597],[268,0.0607],[270,0.0613],[272,0.0646],[276,0.066],[278,0.0718],[280,0.0771],[282,0.0837],[284,0.104],[286,0.1084],[288,0.1102],[290,0.1053],[292,0.0995],[294,0.0983],[296,0.1007],[298,0.1024],[300,0.1034],[302,0.1029],[304,0.101],[306,0.1025],[308,0.1044],[310,0.1027],[314,0.1005],[316,0.0971],[318,0.0972],[320,0.0947],[322,0.0959],[324,0.0984],[326,0.0993],[328,0.1014],[330,0.1015],[334,0.1057],[336,0.0671],[338,0.0664],[340,0.0655],[342,0.0647],[344,0.0638],[346,0.0625],[348,0.0616],[350,0.0603],[352,0.0597],[354,0.0596],[356,0.0568],[358,0.0546],[360,0.0531],[362,0.0482],[366,0.0397],[368,0.0369],[370,0.0347],[372,0.0318],[374,0.0281],[376,0.0275],[378,0.0244],[380,0.0202],[384,0.0201],[386,0.0207],[388,0.0253],[390,0.0264],[392,0.0265],[396,0.0169],[398,0.0155],[400,0.0153],[402,0.026],[404,0.0258],[406,0.0284],[408,0.0274],[410,0.0192],[412,0.0226],[414,0.0216],[416,0.0196],[418,0.0183],[420,0.0182],[422,0.019],[424,0.0188],[426,0.0187],[430,0.0174],[432,0.0176],[434,0.0168],[436,0.0169],[438,0.0173],[440,0.017],[442,0.0195],[446,0.0187],[448,0.0174],[450,0.0152],[452,0.0153],[454,0.0153],[456,0.0162],[458,0.0163],[460,0.0178],[462,0.0186],[464,0.0184],[466,0.0191],[468,0.0172],[470,0.0164],[472,0.0164],[476,0.012],[478,0.0073],[480,0.0043],[482,0.0022],[484,0.0018],[486,0.001],[488,0.0008],[490,0.0008],[492,0.0006],[506,0],[510,0.0003],[518,0],[524,0.0013],[526,0.0021],[528,0.0022],[532,0.0028],[536,0.0027],[538,0.0024],[542,0.0022],[546,0.0015],[552,0],[558,0],[568,0.0004],[588,0.0002],[592,0.0006],[596,0.0009],[602,0.0018],[604,0.0042],[606,0.0048],[608,0.0071],[612,0.0125],[614,0.0157],[618,0.0193],[620,0.0199],[622,0.0194],[624,0.0194],[626,0.02],[628,0.0198],[634,0.0219],[636,0.0233],[638,0.0238],[640,0.0251],[644,0.0255],[646,0.0252],[648,0.0253],[652,0.0261],[654,0.027],[656,0.027],[658,0.0273],[662,0.0284],[664,0.0295],[668,0.0301],[670,0.0313],[672,0.0318],[674,0.033],[676,0.0338],[678,0.0334],[680,0.0336],[682,0.0333],[686,0.0337],[690,0.0333],[694,0.0323],[698,0.0319],[700,0.032],[702,0.0319],[704,0.0307],[706,0.0302],[708,0.0292],[710,0.0287],[712,0.0287],[714,0.0277],[716,0.0261],[718,0.0243],[720,0.0243]]
       }
       CTRL_R_mouth_cornerPull {
        type=MovieSceneScriptingFloatChannel
        default=0
        thinned_from=492
        keys [[-262,0.0015],[-258,0.0009],[-256,0.0003],[-252,0],[-242,0],[-240,0.0002],[-232,0.0004],[-228,0.0007],[-220,0.0004],[-218,0.0002],[-214,0.0007],[-210,0.001],[-206,0.0007],[-198,0.0004],[-188,0.0009],[-186,0.0009],[-182,0.0002],[-170,0.0021],[-150,0.0024],[-144,0.0021],[-140,0.0023],[-122,0.0022],[-114,0.0024],[-104,0.0022],[-76,0.002],[-74,0.0023],[-60,0.0024],[-54,0.0019],[-42,0.0014],[-36,0.0024],[-32,0.0024],[-30,0.0023],[-26,0.0024],[-24,0.0004],[268,0.0003],[270,0.0005],[272,0.0004],[274,0.0007],[276,0.0041],[278,0.0117],[280,0.0218],[282,0.0264],[284,0.035],[286,0.0347],[288,0.0327],[290,0.027],[292,0.0261],[294,0.0259],[296,0.0279],[298,0.0266],[300,0.0262],[302,0.0268],[306,0.0286],[308,0.0283],[310,0.0268],[312,0.0261],[314,0.0262],[316,0.0278],[318,0.0274],[320,0.0262],[322,0.028],[324,0.0302],[326,0.031],[328,0.0303],[330,0.0303],[332,0.0316],[334,0.0314],[336,0.0048],[338,0.0037],[348,0.0007],[350,0.0003],[354,0.0002],[356,0.0005],[360,0.0004],[362,0],[384,0],[386,0.0005],[388,0.0014],[392,0.0016],[394,0.0011],[396,0.0011],[408,0.0019],[410,0.0016],[426,0.0016],[430,0.0013],[436,0.0013],[438,0.0011],[456,0.0016],[468,0.0013],[470,0.0009],[472,0.001],[476,0.0015],[480,0.0006],[486,0.0008],[492,0.0003],[494,0.0004],[518,0],[522,0.001],[528,0.0019],[530,0.0023],[532,0.0025],[536,0.0022],[538,0.0019],[544,0.0015],[546,0.0012],[548,0.0005],[550,0.0002],[556,0],[570,0.0002],[574,0],[580,0.0001],[588,0],[596,0.0006],[598,0.0009],[600,0.0017],[610,0.0017],[614,0.002],[620,0.0019],[622,0.0015],[634,0.0014],[638,0.0011],[642,0.0015],[646,0.0011],[648,0.0011],[652,0.0015],[658,0.0015],[664,0.0017],[670,0.0014],[676,0.0014],[682,0.0011],[690,0.0014],[700,0.001],[702,0.0012],[704,0.0012],[706,0.0014],[714,0.0012],[720,0.0012]]
       }
      ]
     }
     {
      type=MovieSceneControlRigParameterSection
      channels [
       CTRL_C_eyesAim.Location.X {type=MovieSceneScriptingFloatChannel, default=0, const [0,6]}
       CTRL_C_eyesAim.Location.Y {type=MovieSceneScriptingFloatChannel, default=0, const [0,6]}
       CTRL_C_eyesAim.Location.Z {type=MovieSceneScriptingFloatChannel, default=0, const [0,6]}
       CTRL_C_eyesAim.Rotation.X {type=MovieSceneScriptingFloatChannel, default=0, const [0,6]}
       CTRL_C_eyesAim.Rotation.Y {type=MovieSceneScriptingFloatChannel, default=0, const [0,6]}
       CTRL_C_eyesAim.Rotation.Z {type=MovieSceneScriptingFloatChannel, default=0, const [0,6]}
       CTRL_C_eyesAim.Scale.X {type=MovieSceneScriptingFloatChannel, default=0, const [0,6]}
       CTRL_C_eyesAim.Scale.Y {type=MovieSceneScriptingFloatChannel, default=0, const [0,6]}
       CTRL_C_eyesAim.Scale.Z {type=MovieSceneScriptingFloatChannel, default=0, const [0,6]}
       CTRL_L_eyeAim.Location.X {type=MovieSceneScriptingFloatChannel, default=0, const [0,6]}
       CTRL_L_eyeAim.Location.Y {type=MovieSceneScriptingFloatChannel, default=0, const [0,6]}
       CTRL_L_eyeAim.Location.Z {type=MovieSceneScriptingFloatChannel, default=0, const [0,6]}
       CTRL_L_eyeAim.Rotation.X {type=MovieSceneScriptingFloatChannel, default=0, const [0,6]}
       CTRL_L_eyeAim.Rotation.Y {type=MovieSceneScriptingFloatChannel, default=0, const [0,6]}
       CTRL_L_eyeAim.Rotation.Z {type=MovieSceneScriptingFloatChannel, default=0, const [0,6]}
       CTRL_L_eyeAim.Scale.X {type=MovieSceneScriptingFloatChannel, default=0, const [0,6]}
       CTRL_L_eyeAim.Scale.Y {type=MovieSceneScriptingFloatChannel, default=0, const [0,6]}
       CTRL_L_eyeAim.Scale.Z {type=MovieSceneScriptingFloatChannel, default=0, const [0,6]}
       CTRL_R_eyeAim.Location.X {type=MovieSceneScriptingFloatChannel, default=0, const [0,6]}
       CTRL_R_eyeAim.Location.Y {type=MovieSceneScriptingFloatChannel, default=0, const [0,6]}
       CTRL_R_eyeAim.Location.Z {type=MovieSceneScriptingFloatChannel, default=0, const [0,6]}
       CTRL_R_eyeAim.Rotation.X {type=MovieSceneScriptingFloatChannel, default=0, const [0,6]}
       CTRL_R_eyeAim.Rotation.Y {type=MovieSceneScriptingFloatChannel, default=0, const [0,6]}
       CTRL_R_eyeAim.Rotation.Z {type=MovieSceneScriptingFloatChannel, default=0, const [0,6]}
       CTRL_R_eyeAim.Scale.X {type=MovieSceneScriptingFloatChannel, default=0, const [0,6]}
       CTRL_R_eyeAim.Scale.Y {type=MovieSceneScriptingFloatChannel, default=0, const [0,6]}
       CTRL_R_eyeAim.Scale.Z {type=MovieSceneScriptingFloatChannel, default=0, const [0,6]}
       CTRL_convergenceSwitch {type=MovieSceneScriptingFloatChannel, default=0, const [0,6]}
       CTRL_Master.Location.X {type=MovieSceneScriptingFloatChannel, default=0, const [0,6]}
       CTRL_Master.Location.Y {type=MovieSceneScriptingFloatChannel, default=0, const [0,6]}
       CTRL_Master.Location.Z {type=MovieSceneScriptingFloatChannel, default=0, const [0,6]}
       CTRL_Master.Rotation.X {type=MovieSceneScriptingFloatChannel, default=0, const [0,6]}
       CTRL_Master.Rotation.Y {type=MovieSceneScriptingFloatChannel, default=0, const [0,6]}
       CTRL_Master.Rotation.Z {type=MovieSceneScriptingFloatChannel, default=0, const [0,6]}
       CTRL_Master.Scale.X {type=MovieSceneScriptingFloatChannel, default=0, const [0,6]}
       CTRL_Master.Scale.Y {type=MovieSceneScriptingFloatChannel, default=0, const [0,6]}
       CTRL_Master.Scale.Z {type=MovieSceneScriptingFloatChannel, default=0, const [0,6]}
       CTRL_C_jaw.X {type=MovieSceneScriptingFloatChannel, default=0, thinned_from=8, keys [[260,0],[338,0],[348,-0.0059],[364,-0.0059]]}
       CTRL_C_jaw.Y {type=MovieSceneScriptingFloatChannel, default=0, thinned_from=8, keys [[260,0],[338,0],[348,-0.0492],[364,-0.0492]]}
       CTRL_C_jaw_openExtreme {type=MovieSceneScriptingFloatChannel, default=0, thinned_from=8, keys [[260,0],[338,0],[348,-0.0015],[364,-0.0015]]}
       CTRL_C_jaw_fwdBack {type=MovieSceneScriptingFloatChannel, default=0, thinned_from=8, keys [[260,0],[338,0],[348,-0.0645],[364,-0.0645]]}
       CTRL_R_jaw_ChinRaiseD {type=MovieSceneScriptingFloatChannel, default=0, thinned_from=8, keys [[260,0],[338,0],[348,-0.1288],[364,-0.1288]]}
       CTRL_R_jaw_ChinRaiseU {type=MovieSceneScriptingFloatChannel, default=0, const [0,8]}
       CTRL_L_jaw_ChinRaiseU {type=MovieSceneScriptingFloatChannel, default=0, const [0,8]}
       CTRL_L_jaw_ChinRaiseD {type=MovieSceneScriptingFloatChannel, default=0, thinned_from=8, keys [[260,0],[338,0],[348,-0.1325],[364,-0.1325]]}
       CTRL_R_mouth_lipsBlow {type=MovieSceneScriptingFloatChannel, default=0, const [0,8]}
       CTRL_L_mouth_lipsTogetherU {type=MovieSceneScriptingFloatChannel, default=0, thinned_from=8, keys [[260,0],[338,0],[348,-0.0294],[364,-0.0294]]}
       CTRL_L_mouth_lipsTogetherD {type=MovieSceneScriptingFloatChannel, default=0, thinned_from=8, keys [[260,0],[338,0],[348,-0.0245],[364,-0.0245]]}
       CTRL_R_mouth_lipsTogetherU {type=MovieSceneScriptingFloatChannel, default=0, thinned_from=8, keys [[260,0],[338,0],[348,-0.0294],[364,-0.0294]]}
       CTRL_L_mouth_lipsBlow {type=MovieSceneScriptingFloatChannel, default=0, const [0,8]}
       CTRL_R_mouth_lipsTogetherD {type=MovieSceneScriptingFloatChannel, default=0, thinned_from=8, keys [[260,0],[338,0],[348,-0.0294],[364,-0.0294]]}
       CTRL_C_tongue_roll.X {type=MovieSceneScriptingFloatChannel, default=0, const [0,8]}
       CTRL_C_tongue_roll.Y {type=MovieSceneScriptingFloatChannel, default=0, const [0,8]}
       CTRL_C_tongue.X {type=MovieSceneScriptingFloatChannel, default=0, const [0,8]}
       CTRL_C_tongue.Y {type=MovieSceneScriptingFloatChannel, default=0, const [0,8]}
       CTRL_C_tongue_narrowWide {type=MovieSceneScriptingFloatChannel, default=0, const [0,8]}
       CTRL_C_tongue_inOut {type=MovieSceneScriptingFloatChannel, default=0, const [0,8]}
       CTRL_R_nose.X {type=MovieSceneScriptingFloatChannel, default=0, thinned_from=7, keys [[260,0],[338,0],[348,0.0058],[364,0.0058]]}
       CTRL_R_nose.Y {type=MovieSceneScriptingFloatChannel, default=0, thinned_from=7, keys [[260,0],[338,0],[348,-0.0146],[364,-0.0146]]}
       CTRL_L_nose.X {type=MovieSceneScriptingFloatChannel, default=0, thinned_from=7, keys [[260,0],[338,0],[348,0.0134],[364,0.0134]]}
       CTRL_L_nose.Y {type=MovieSceneScriptingFloatChannel, default=0, thinned_from=7, keys [[260,0],[338,0],[348,0.0242],[364,0.0242]]}
       CTRL_L_eye.X {
        type=MovieSceneScriptingFloatChannel
        default=0
        thinned_from=9
        keys [[144,-0.2715],[180,-0.3409],[328,-0.3409],[338,-0.093],[348,0.1636],[364,0.1636]]
       }
       CTRL_L_eye.Y {
        type=MovieSceneScriptingFloatChannel
        default=0
        thinned_from=9
        keys [[144,-0.4023],[180,-0.4836],[260,-0.4566],[328,-0.4561],[338,-0.5456],[348,-0.7856],[364,-0.7856]]
       }
       CTRL_L_eye_blink {
        type=MovieSceneScriptingFloatChannel
        default=0
        thinned_from=20
        keys [[94,0],[98,1.4777],[102,1.4777],[105,0.5371],[113,0.0859],[119,0],[141,0],[143,0.4666],[145,1.4777],[147,0.8054],[151,0.2303],[157,0],[328,0],[332,1.4777],[338,0],[348,0.4732],[356,0.5173],[360,1.5336],[364,0.5329]]
       }
       CTRL_R_eye.X {
        type=MovieSceneScriptingFloatChannel
        default=0
        thinned_from=9
        keys [[144,-0.1467],[180,-0.2161],[328,-0.2161],[338,-0.093],[348,0.1773],[360,0.1733],[364,0.1733]]
       }
       CTRL_R_eye.Y {
        type=MovieSceneScriptingFloatChannel
        default=0
        thinned_from=9
        keys [[144,-0.4034],[180,-0.485],[260,-0.4569],[328,-0.4564],[338,-0.5456],[348,-0.7866],[364,-0.7866]]
       }
       CTRL_R_eye_blink {
        type=MovieSceneScriptingFloatChannel
        default=0
        thinned_from=20
        keys [[94,0],[98,1.6085],[102,1.6085],[105,0.5846],[113,0.0935],[119,0],[141,0],[143,0.5079],[145,1.6085],[147,0.8767],[151,0.2506],[157,0],[328,0],[332,1.6085],[338,0],[348,0.6164],[356,0.6796],[360,1.6855],[364,0.688]]
       }
       CTRL_C_eye.X {
        type=MovieSceneScriptingFloatChannel
        default=0.3646
        thinned_from=18
        keys [[90,0.2622],[120,0.2635],[149,0.2617],[180,0.4163],[213,0.4163],[260,0.2325],[270,0.2325],[290,0.2306],[324,0.2306],[326,-0.3855],[338,-0.3487],[348,-0.6871],[356,-0.6871],[360,-0.4838],[364,-0.181],[366,-0.1097]]
       }
       CTRL_C_eye.Y {
        type=MovieSceneScriptingFloatChannel
        default=-1
        thinned_from=18
        keys [[61,-0.2859],[110,-0.0209],[138,-0.6519],[180,-0.7032],[328,-0.7032],[338,-0.6873],[348,-0.5279],[356,-0.5279],[360,-0.1146],[364,0.1279],[366,0.2204]]
       }
       CTRL_C_eyesConvergence {type=MovieSceneScriptingFloatChannel, default=0, const [0,7]}
       CTRL_L_eye_cheekRaise {type=MovieSceneScriptingFloatChannel, default=0, thinned_from=7, keys [[260,0],[338,0],[348,-0.019],[364,-0.019]]}
       CTRL_L_eye_squintInner {type=MovieSceneScriptingFloatChannel, default=0, thinned_from=7, keys [[260,0],[338,0],[348,-0.0632],[364,-0.0632]]}
       CTRL_L_nose_wrinkleUpper {type=MovieSceneScriptingFloatChannel, default=0, thinned_from=7, keys [[260,0],[338,0],[348,-1],[364,-1]]}
       CTRL_R_nose_wrinkleUpper {type=MovieSceneScriptingFloatChannel, default=0, thinned_from=7, keys [[260,0],[338,0],[348,-1],[364,-1]]}
       CTRL_R_eye_squintInner {type=MovieSceneScriptingFloatChannel, default=0, thinned_from=7, keys [[260,0],[338,0],[348,-0.0614],[364,-0.0614]]}
       CTRL_R_eye_cheekRaise {type=MovieSceneScriptingFloatChannel, default=0, thinned_from=7, keys [[260,0],[364,0]]}
       CTRL_C_eyelidsReduce {type=MovieSceneScriptingFloatChannel, default=0, const [0,7]}
       CTRL_R_brow_lateral {type=MovieSceneScriptingFloatChannel, default=0, thinned_from=7, keys [[260,0],[338,0],[348,-0.0011],[360,-0.0011],[364,0]]}
       CTRL_L_brow_lateral {type=MovieSceneScriptingFloatChannel, default=0, thinned_from=7, keys [[260,0],[338,0],[348,-0.0046],[360,-0.0046],[364,0]]}
       CTRL_R_brow_down {type=MovieSceneScriptingFloatChannel, default=0, thinned_from=7, keys [[260,0],[338,0],[348,-0.0019],[360,-0.0019],[364,0]]}
       CTRL_L_brow_down {type=MovieSceneScriptingFloatChannel, default=0, thinned_from=7, keys [[260,0],[338,0],[348,-0.004],[360,-0.004],[364,0]]}
       CTRL_L_brow_raiseOut {type=MovieSceneScriptingFloatChannel, default=0, thinned_from=7, keys [[260,0],[338,0],[348,-0.841],[360,-0.841],[364,0]]}
       CTRL_L_brow_raiseIn {type=MovieSceneScriptingFloatChannel, default=0, thinned_from=7, keys [[260,0],[338,0],[348,-0.5337],[360,-0.5337],[364,0]]}
       CTRL_R_brow_raiseIn {type=MovieSceneScriptingFloatChannel, default=0, thinned_from=7, keys [[260,0],[338,0],[348,-0.5148],[360,-0.5148],[364,0]]}
       CTRL_R_brow_raiseOut {type=MovieSceneScriptingFloatChannel, default=0, thinned_from=7, keys [[260,0],[338,0],[348,-0.7905],[360,-0.7905],[364,0]]}
       CTRL_lookAtSwitch {type=MovieSceneScriptingFloatChannel, default=0, const [0,8]}
       CTRL_C_teethD.X {type=MovieSceneScriptingFloatChannel, default=0, const [0,8]}
       CTRL_C_teethD.Y {type=MovieSceneScriptingFloatChannel, default=0, const [0,8]}
       CTRL_C_teethU.X {type=MovieSceneScriptingFloatChannel, default=0, const [0,8]}
       CTRL_C_teethU.Y {type=MovieSceneScriptingFloatChannel, default=0, const [0,8]}
       CTRL_C_teeth_fwdBackU {type=MovieSceneScriptingFloatChannel, default=0, const [0,8]}
       CTRL_L_mouth_corner.X {type=MovieSceneScriptingFloatChannel, default=0, const [0,8]}
       CTRL_L_mouth_corner.Y {type=MovieSceneScriptingFloatChannel, default=0, thinned_from=8, keys [[260,0],[338,0],[348,-0.0368],[364,-0.0368]]}
       CTRL_R_mouth_corner.X {type=MovieSceneScriptingFloatChannel, default=0, const [0,8]}
       CTRL_R_mouth_corner.Y {type=MovieSceneScriptingFloatChannel, default=0, thinned_from=8, keys [[260,0],[338,0],[348,-0.0251],[364,-0.0251]]}
       CTRL_C_teeth_fwdBackD {type=MovieSceneScriptingFloatChannel, default=0, const [0,8]}
       CTRL_R_mouth_cornerSharpnessD {type=MovieSceneScriptingFloatChannel, default=0, thinned_from=8, keys [[260,0],[338,0],[348,-0.0002],[364,-0.0002]]}
       CTRL_R_mouth_cornerSharpnessU {type=MovieSceneScriptingFloatChannel, default=0, thinned_from=8, keys [[260,0],[338,0],[348,-0.0073],[364,-0.0073]]}
       CTRL_L_eye_eyelidU.X {type=MovieSceneScriptingFloatChannel, default=0, const [0,7]}
       CTRL_L_eye_eyelidU.Y {type=MovieSceneScriptingFloatChannel, default=0, const [0,7]}
       CTRL_L_mouth_cornerSharpnessD {type=MovieSceneScriptingFloatChannel, default=0, thinned_from=8, keys [[260,0],[338,0],[348,-0.0103],[364,-0.0103]]}
       CTRL_L_eye_eyelidD.X {type=MovieSceneScriptingFloatChannel, default=0, const [0,7]}
       CTRL_L_eye_eyelidD.Y {type=MovieSceneScriptingFloatChannel, default=0, const [0,7]}
       CTRL_L_mouth_cornerSharpnessU {type=MovieSceneScriptingFloatChannel, default=0, thinned_from=8, keys [[260,0],[338,0],[348,-0.0142],[364,-0.0142]]}
       CTRL_R_mouth_lipsTowardsTeethD {type=MovieSceneScriptingFloatChannel, default=0, const [0,8]}
       CTRL_L_mouth_lipsTowardsTeethD {type=MovieSceneScriptingFloatChannel, default=0, const [0,8]}
       CTRL_R_eye_eyelidU.X {type=MovieSceneScriptingFloatChannel, default=0, const [0,7]}
       CTRL_R_eye_eyelidU.Y {type=MovieSceneScriptingFloatChannel, default=0, const [0,7]}
       CTRL_L_mouth_lipsTowardsTeethU {type=MovieSceneScriptingFloatChannel, default=0, const [0,8]}
       CTRL_R_eye_eyelidD.X {type=MovieSceneScriptingFloatChannel, default=0, const [0,7]}
       CTRL_R_eye_eyelidD.Y {type=MovieSceneScriptingFloatChannel, default=0, const [0,7]}
       CTRL_R_mouth_lipsTowardsTeethU {type=MovieSceneScriptingFloatChannel, default=0, const [0,8]}
       CTRL_faceGUIfollowHead {type=MovieSceneScriptingFloatChannel, default=0, const [0,6]}
       CTRL_eyesAimFollowHead {type=MovieSceneScriptingFloatChannel, default=0, const [0,6]}
       CTRL_C_nose.X {type=MovieSceneScriptingFloatChannel, default=0, const [0,7]}
       CTRL_C_nose.Y {type=MovieSceneScriptingFloatChannel, default=0, const [0,7]}
       CTRL_C_teeth_rotateU {type=MovieSceneScriptingFloatChannel, default=0, const [0,8]}
       CTRL_C_teeth_rotateD {type=MovieSceneScriptingFloatChannel, default=0, const [0,8]}
       CTRL_L_brow_inner {type=MovieSceneScriptingFloatChannel, default=0, const [0,7]}
       CTRL_R_brow_inner {type=MovieSceneScriptingFloatChannel, default=0, const [0,7]}
       CTRL_L_brow_outer {type=MovieSceneScriptingFloatChannel, default=0, const [0,7]}
       CTRL_R_brow_outer {type=MovieSceneScriptingFloatChannel, default=0, const [0,7]}
       CTRL_C_mouth.X {type=MovieSceneScriptingFloatChannel, default=0, thinned_from=8, keys [[260,0],[328,0],[338,-0.0179],[348,-0.0215],[364,-0.0215]]}
       CTRL_C_mouth.Y {type=MovieSceneScriptingFloatChannel, default=0, const [0,8]}
       CTRL_L_mouth_upperLipRaise {type=MovieSceneScriptingFloatChannel, default=0, thinned_from=8, keys [[260,0],[328,0],[338,-0.0044],[348,-0.0053],[364,-0.0053]]}
       CTRL_L_mouth_cornerPull {
        type=MovieSceneScriptingFloatChannel
        default=0.1087
        thinned_from=9
        keys [[144,0.1643],[260,0.2747],[286,0.2764],[328,0.2767],[338,-0.2275],[348,-0.2948],[364,-0.2948]]
       }
       CTRL_L_mouth_dimple {type=MovieSceneScriptingFloatChannel, default=0, thinned_from=9, keys [[144,0.6941],[328,0.6941],[338,-0.125],[348,-0.2888],[364,-0.2888]]}
       CTRL_L_mouth_cornerDepress {type=MovieSceneScriptingFloatChannel, default=0, thinned_from=8, keys [[260,0],[328,0],[338,-0.0029],[348,-0.0035],[364,-0.0035]]}
       CTRL_L_mouth_stretch {type=MovieSceneScriptingFloatChannel, default=0, thinned_from=8, keys [[260,0],[328,0],[338,-0.024],[348,-0.0288],[364,-0.0288]]}
       CTRL_L_mouth_lowerLipDepress {type=MovieSceneScriptingFloatChannel, default=0, thinned_from=8, keys [[260,0],[328,0],[338,-0.0011],[348,-0.0013],[364,-0.0013]]}
       CTRL_R_mouth_lowerLipDepress {type=MovieSceneScriptingFloatChannel, default=0, thinned_from=8, keys [[260,0],[328,0],[338,-0.0015],[348,-0.0018],[364,-0.0018]]}
       CTRL_R_mouth_stretch {type=MovieSceneScriptingFloatChannel, default=0, thinned_from=8, keys [[260,0],[328,0],[338,-0.0191],[348,-0.0229],[364,-0.0229]]}
       CTRL_L_mouth_towardsD {type=MovieSceneScriptingFloatChannel, default=0, thinned_from=8, keys [[260,0.0001],[338,0.0001],[348,-0.0029],[364,-0.0029]]}
       CTRL_R_mouth_funnelU {type=MovieSceneScriptingFloatChannel, default=0, thinned_from=8, keys [[260,0],[338,0],[348,-0.0037],[364,-0.0037]]}
       CTRL_L_mouth_towardsU {type=MovieSceneScriptingFloatChannel, default=0, thinned_from=8, keys [[260,0],[338,0],[348,-0.0088],[364,-0.0088]]}
       CTRL_R_mouth_lipBiteD {type=MovieSceneScriptingFloatChannel, default=0, thinned_from=8, keys [[260,0],[338,0],[348,-0.0144],[364,-0.0144]]}
       CTRL_R_mouth_funnelD {type=MovieSceneScriptingFloatChannel, default=0, thinned_from=8, keys [[260,0],[338,0],[348,-0.0018],[364,-0.0018]]}
       CTRL_L_mouth_pressU {type=MovieSceneScriptingFloatChannel, default=0, const [0,8]}
       CTRL_L_mouth_funnelD {type=MovieSceneScriptingFloatChannel, default=0, thinned_from=8, keys [[260,0],[338,0],[348,-0.0008],[364,-0.0008]]}
       CTRL_L_mouth_suckBlow {type=MovieSceneScriptingFloatChannel, default=0, thinned_from=8, keys [[260,0],[338,0],[348,-0.0049],[364,-0.0049]]}
       CTRL_L_mouth_pressD {type=MovieSceneScriptingFloatChannel, default=0, const [0,8]}
       CTRL_L_mouth_funnelU {type=MovieSceneScriptingFloatChannel, default=0, thinned_from=8, keys [[260,0.0001],[338,0],[348,-0.0029],[364,-0.0029]]}
       CTRL_R_mouth_towardsU {type=MovieSceneScriptingFloatChannel, default=-0.0001, thinned_from=8, keys [[260,-0.0001],[338,-0.0001],[348,-0.013],[364,-0.013]]}
       CTRL_R_mouth_purseD {type=MovieSceneScriptingFloatChannel, default=0, thinned_from=8, keys [[260,0],[338,0],[348,-0.0191],[364,-0.0191]]}
       CTRL_L_mouth_lipBiteU {type=MovieSceneScriptingFloatChannel, default=0, thinned_from=8, keys [[260,0],[338,0],[348,-0.0253],[364,-0.0253]]}
       CTRL_R_mouth_pressD {type=MovieSceneScriptingFloatChannel, default=0, const [0,8]}
       CTRL_R_mouth_pressU {type=MovieSceneScriptingFloatChannel, default=0, const [0,8]}
       CTRL_L_mouth_purseD {type=MovieSceneScriptingFloatChannel, default=0, thinned_from=8, keys [[260,0],[338,0],[348,-0.0372],[364,-0.0372]]}
       CTRL_R_mouth_lipBiteU {type=MovieSceneScriptingFloatChannel, default=0, thinned_from=8, keys [[260,0],[338,0],[348,-0.02],[364,-0.02]]}
       CTRL_R_mouth_suckBlow {type=MovieSceneScriptingFloatChannel, default=0, thinned_from=8, keys [[260,0],[338,0],[348,-0.0117],[364,-0.0117]]}
       CTRL_L_mouth_purseU {type=MovieSceneScriptingFloatChannel, default=0, thinned_from=8, keys [[260,0],[338,0],[348,-0.0499],[364,-0.0499]]}
       CTRL_R_mouth_towardsD {type=MovieSceneScriptingFloatChannel, default=0, thinned_from=8, keys [[260,0],[338,0],[348,-0.0025],[364,-0.0025]]}
       CTRL_L_mouth_lipBiteD {type=MovieSceneScriptingFloatChannel, default=0, thinned_from=8, keys [[260,0],[338,0],[348,-0.0149],[364,-0.0149]]}
       CTRL_R_mouth_purseU {type=MovieSceneScriptingFloatChannel, default=0, thinned_from=8, keys [[260,0],[338,0],[348,-0.095],[364,-0.095]]}
       CTRL_R_mouth_cornerDepress {type=MovieSceneScriptingFloatChannel, default=0, thinned_from=8, keys [[260,0],[328,0],[338,-0.0016],[348,-0.002],[364,-0.002]]}
       CTRL_R_mouth_upperLipRaise {type=MovieSceneScriptingFloatChannel, default=0, thinned_from=8, keys [[260,0],[328,0],[338,-0.0011],[348,-0.0013],[364,-0.0013]]}
       CTRL_R_mouth_dimple {
        type=MovieSceneScriptingFloatChannel
        default=0.8943
        thinned_from=10
        keys [[104,0.8986],[328,0.8986],[338,0.061],[348,-0.1057],[364,-0.1057]]
       }
       CTRL_R_mouth_cornerPull {
        type=MovieSceneScriptingFloatChannel
        default=0
        thinned_from=10
        keys [[104,0.4261],[144,0.427],[260,0.53],[286,0.5301],[296,0.5318],[328,0.563],[338,-0.0263],[348,-0.0315],[364,-0.0315]]
       }
      ]
     }
     {
      type=MovieSceneControlRigParameterSection
      channels [
       CTRL_C_eyesAim.Location.X {type=MovieSceneScriptingFloatChannel, default=0, const [0,1]}
       CTRL_C_eyesAim.Location.Y {type=MovieSceneScriptingFloatChannel, default=0, const [0,1]}
       CTRL_C_eyesAim.Location.Z {type=MovieSceneScriptingFloatChannel, default=0, const [0,1]}
       CTRL_C_eyesAim.Rotation.X {type=MovieSceneScriptingFloatChannel, default=0, const [0,1]}
       CTRL_C_eyesAim.Rotation.Y {type=MovieSceneScriptingFloatChannel, default=0, const [0,1]}
       CTRL_C_eyesAim.Rotation.Z {type=MovieSceneScriptingFloatChannel, default=0, const [0,1]}
       CTRL_C_eyesAim.Scale.X {type=MovieSceneScriptingFloatChannel, default=0, const [0,1]}
       CTRL_C_eyesAim.Scale.Y {type=MovieSceneScriptingFloatChannel, default=0, const [0,1]}
       CTRL_C_eyesAim.Scale.Z {type=MovieSceneScriptingFloatChannel, default=0, const [0,1]}
       CTRL_L_eyeAim.Location.X {type=MovieSceneScriptingFloatChannel, default=0, const [0,1]}
       CTRL_L_eyeAim.Location.Y {type=MovieSceneScriptingFloatChannel, default=0, const [0,1]}
       CTRL_L_eyeAim.Location.Z {type=MovieSceneScriptingFloatChannel, default=0, const [0,1]}
       CTRL_L_eyeAim.Rotation.X {type=MovieSceneScriptingFloatChannel, default=0, const [0,1]}
       CTRL_L_eyeAim.Rotation.Y {type=MovieSceneScriptingFloatChannel, default=0, const [0,1]}
       CTRL_L_eyeAim.Rotation.Z {type=MovieSceneScriptingFloatChannel, default=0, const [0,1]}
       CTRL_L_eyeAim.Scale.X {type=MovieSceneScriptingFloatChannel, default=0, const [0,1]}
       CTRL_L_eyeAim.Scale.Y {type=MovieSceneScriptingFloatChannel, default=0, const [0,1]}
       CTRL_L_eyeAim.Scale.Z {type=MovieSceneScriptingFloatChannel, default=0, const [0,1]}
       CTRL_R_eyeAim.Location.X {type=MovieSceneScriptingFloatChannel, default=0, const [0,1]}
       CTRL_R_eyeAim.Location.Y {type=MovieSceneScriptingFloatChannel, default=0, const [0,1]}
       CTRL_R_eyeAim.Location.Z {type=MovieSceneScriptingFloatChannel, default=0, const [0,1]}
       CTRL_R_eyeAim.Rotation.X {type=MovieSceneScriptingFloatChannel, default=0, const [0,1]}
       CTRL_R_eyeAim.Rotation.Y {type=MovieSceneScriptingFloatChannel, default=0, const [0,1]}
       CTRL_R_eyeAim.Rotation.Z {type=MovieSceneScriptingFloatChannel, default=0, const [0,1]}
       CTRL_R_eyeAim.Scale.X {type=MovieSceneScriptingFloatChannel, default=0, const [0,1]}
       CTRL_R_eyeAim.Scale.Y {type=MovieSceneScriptingFloatChannel, default=0, const [0,1]}
       CTRL_R_eyeAim.Scale.Z {type=MovieSceneScriptingFloatChannel, default=0, const [0,1]}
       CTRL_convergenceSwitch {type=MovieSceneScriptingFloatChannel, default=0, const [0,1]}
       CTRL_Master.Location.X {type=MovieSceneScriptingFloatChannel, default=0, const [0,1]}
       CTRL_Master.Location.Y {type=MovieSceneScriptingFloatChannel, default=0, const [0,1]}
       CTRL_Master.Location.Z {type=MovieSceneScriptingFloatChannel, default=0, const [0,1]}
       CTRL_Master.Rotation.X {type=MovieSceneScriptingFloatChannel, default=0, const [0,1]}
       CTRL_Master.Rotation.Y {type=MovieSceneScriptingFloatChannel, default=0, const [0,1]}
       CTRL_Master.Rotation.Z {type=MovieSceneScriptingFloatChannel, default=0, const [0,1]}
       CTRL_Master.Scale.X {type=MovieSceneScriptingFloatChannel, default=0, const [0,1]}
       CTRL_Master.Scale.Y {type=MovieSceneScriptingFloatChannel, default=0, const [0,1]}
       CTRL_Master.Scale.Z {type=MovieSceneScriptingFloatChannel, default=0, const [0,1]}
       CTRL_C_jaw.X {type=MovieSceneScriptingFloatChannel, default=0, keys [[296,-0.0869],[316,-0.0869],[350,0]]}
       CTRL_C_jaw.Y {type=MovieSceneScriptingFloatChannel, default=0, keys [[296,-0.0472],[316,-0.0472],[350,0]]}
       CTRL_C_jaw_openExtreme {type=MovieSceneScriptingFloatChannel, default=0.0003, keys [[296,-0.0007],[316,-0.0007],[350,0.0003]]}
       CTRL_C_jaw_fwdBack {type=MovieSceneScriptingFloatChannel, default=0, keys [[296,-0.0373],[316,-0.0373],[350,0]]}
       CTRL_R_jaw_ChinRaiseD {type=MovieSceneScriptingFloatChannel, default=0.0139, keys [[296,-0.0705],[316,-0.0705],[350,0.0139]]}
       CTRL_R_jaw_ChinRaiseU {type=MovieSceneScriptingFloatChannel, default=0, const [0,3]}
       CTRL_L_jaw_ChinRaiseU {type=MovieSceneScriptingFloatChannel, default=0, const [0,3]}
       CTRL_L_jaw_ChinRaiseD {type=MovieSceneScriptingFloatChannel, default=0.0161, keys [[296,-0.0716],[316,-0.0716],[350,0.0161]]}
       CTRL_R_mouth_lipsBlow {type=MovieSceneScriptingFloatChannel, default=0, const [0,3]}
       CTRL_L_mouth_lipsTogetherU {type=MovieSceneScriptingFloatChannel, default=0, keys [[296,-0.0088],[316,-0.0088],[350,0]]}
       CTRL_L_mouth_lipsTogetherD {type=MovieSceneScriptingFloatChannel, default=0.0101, keys [[296,-0.0072],[316,-0.0072],[350,0.0101]]}
       CTRL_R_mouth_lipsTogetherU {type=MovieSceneScriptingFloatChannel, default=0, keys [[296,-0.0087],[316,-0.0087],[350,0]]}
       CTRL_L_mouth_lipsBlow {type=MovieSceneScriptingFloatChannel, default=0, const [0,3]}
       CTRL_R_mouth_lipsTogetherD {type=MovieSceneScriptingFloatChannel, default=0.0179, keys [[296,-0.0038],[316,-0.0038],[350,0.0179]]}
       CTRL_C_tongue_roll.X {type=MovieSceneScriptingFloatChannel, default=0, const [0,3]}
       CTRL_C_tongue_roll.Y {type=MovieSceneScriptingFloatChannel, default=0, const [0,3]}
       CTRL_C_tongue.X {type=MovieSceneScriptingFloatChannel, default=0, const [0,3]}
       CTRL_C_tongue.Y {type=MovieSceneScriptingFloatChannel, default=0, const [0,3]}
       CTRL_C_tongue_narrowWide {type=MovieSceneScriptingFloatChannel, default=0, const [0,3]}
       CTRL_C_tongue_inOut {type=MovieSceneScriptingFloatChannel, default=0, const [0,3]}
       CTRL_R_nose.X {type=MovieSceneScriptingFloatChannel, default=0, keys [[316,0.0044],[350,0]]}
       CTRL_R_nose.Y {type=MovieSceneScriptingFloatChannel, default=0, keys [[316,0.0076],[350,0]]}
       CTRL_L_nose.X {type=MovieSceneScriptingFloatChannel, default=0, keys [[316,0.0091],[350,0]]}
       CTRL_L_nose.Y {type=MovieSceneScriptingFloatChannel, default=0, keys [[316,0.0263],[350,0]]}
       CTRL_L_eye.X {type=MovieSceneScriptingFloatChannel, default=0, keys [[275,0.3013],[316,0.3013],[350,0]]}
       CTRL_L_eye.Y {type=MovieSceneScriptingFloatChannel, default=0, keys [[275,-0.3504],[316,-0.3504],[350,0]]}
       CTRL_L_eye_blink {type=MovieSceneScriptingFloatChannel, default=0, keys [[316,0.2771],[350,0]]}
       CTRL_R_eye.X {type=MovieSceneScriptingFloatChannel, default=0, keys [[316,0.3119],[350,0]]}
       CTRL_R_eye.Y {type=MovieSceneScriptingFloatChannel, default=0, keys [[316,-0.3511],[350,0]]}
       CTRL_R_eye_blink {type=MovieSceneScriptingFloatChannel, default=0, keys [[316,0.3465],[350,0]]}
       CTRL_C_eye.X {type=MovieSceneScriptingFloatChannel, default=0, thinned_from=4, keys [[90,-0.1295],[316,-0.1295],[350,0]]}
       CTRL_C_eye.Y {type=MovieSceneScriptingFloatChannel, default=0, thinned_from=4, keys [[90,0.4651],[316,0.4651],[350,0]]}
       CTRL_C_eyesConvergence {type=MovieSceneScriptingFloatChannel, default=0, const [0,2]}
       CTRL_L_eye_cheekRaise {type=MovieSceneScriptingFloatChannel, default=0.0117, keys [[316,-0.0126],[350,0.0117]]}
       CTRL_L_eye_squintInner {type=MovieSceneScriptingFloatChannel, default=0.0522, keys [[316,-0.0493],[350,0.0522]]}
       CTRL_L_nose_wrinkleUpper {type=MovieSceneScriptingFloatChannel, default=0, keys [[316,-0.5614],[350,0]]}
       CTRL_R_nose_wrinkleUpper {type=MovieSceneScriptingFloatChannel, default=0, keys [[316,-0.5614],[350,0]]}
       CTRL_R_eye_squintInner {type=MovieSceneScriptingFloatChannel, default=0.0326, keys [[316,-0.0509],[350,0.0326]]}
       CTRL_R_eye_cheekRaise {type=MovieSceneScriptingFloatChannel, default=0, keys [[316,0],[350,0]]}
       CTRL_C_eyelidsReduce {type=MovieSceneScriptingFloatChannel, default=0, const [0,2]}
       CTRL_R_brow_lateral {type=MovieSceneScriptingFloatChannel, default=0.0001, keys [[316,-0.0007],[350,0.0001]]}
       CTRL_L_brow_lateral {type=MovieSceneScriptingFloatChannel, default=0, keys [[316,-0.0029],[350,0]]}
       CTRL_R_brow_down {type=MovieSceneScriptingFloatChannel, default=0.0005, keys [[316,-0.0003],[350,0.0005]]}
       CTRL_L_brow_down {type=MovieSceneScriptingFloatChannel, default=0, keys [[316,-0.0019],[350,0]]}
       CTRL_L_brow_raiseOut {type=MovieSceneScriptingFloatChannel, default=0.0088, keys [[316,-0.4856],[350,0.0088]]}
       CTRL_L_brow_raiseIn {type=MovieSceneScriptingFloatChannel, default=0.0199, keys [[316,-0.2963],[350,0.0199]]}
       CTRL_R_brow_raiseIn {type=MovieSceneScriptingFloatChannel, default=0.0285, keys [[316,-0.279],[350,0.0285]]}
       CTRL_R_brow_raiseOut {type=MovieSceneScriptingFloatChannel, default=0.0303, keys [[316,-0.4452],[350,0.0303]]}
       CTRL_lookAtSwitch {type=MovieSceneScriptingFloatChannel, default=0, const [0,3]}
       CTRL_C_teethD.X {type=MovieSceneScriptingFloatChannel, default=0, const [0,3]}
       CTRL_C_teethD.Y {type=MovieSceneScriptingFloatChannel, default=0, const [0,3]}
       CTRL_C_teethU.X {type=MovieSceneScriptingFloatChannel, default=0, const [0,3]}
       CTRL_C_teethU.Y {type=MovieSceneScriptingFloatChannel, default=0, const [0,3]}
       CTRL_C_teeth_fwdBackU {type=MovieSceneScriptingFloatChannel, default=0, const [0,3]}
       CTRL_L_mouth_corner.X {type=MovieSceneScriptingFloatChannel, default=0, const [0,3]}
       CTRL_L_mouth_corner.Y {type=MovieSceneScriptingFloatChannel, default=0, keys [[296,-0.0208],[316,-0.0208],[350,0]]}
       CTRL_R_mouth_corner.X {type=MovieSceneScriptingFloatChannel, default=0, const [0,3]}
       CTRL_R_mouth_corner.Y {type=MovieSceneScriptingFloatChannel, default=0, keys [[296,-0.011],[316,-0.011],[350,0]]}
       CTRL_C_teeth_fwdBackD {type=MovieSceneScriptingFloatChannel, default=0, const [0,3]}
       CTRL_R_mouth_cornerSharpnessD {type=MovieSceneScriptingFloatChannel, default=0, thinned_from=3, keys [[296,-0.0001],[350,0]]}
       CTRL_R_mouth_cornerSharpnessU {type=MovieSceneScriptingFloatChannel, default=0, keys [[296,-0.003],[316,-0.003],[350,0]]}
       CTRL_L_eye_eyelidU.X {type=MovieSceneScriptingFloatChannel, default=0, const [0,3]}
       CTRL_L_eye_eyelidU.Y {type=MovieSceneScriptingFloatChannel, default=0, const [0,3]}
       CTRL_L_mouth_cornerSharpnessD {type=MovieSceneScriptingFloatChannel, default=0, keys [[296,-0.0049],[316,-0.0049],[350,0]]}
       CTRL_L_eye_eyelidD.X {type=MovieSceneScriptingFloatChannel, default=0, const [0,2]}
       CTRL_L_eye_eyelidD.Y {type=MovieSceneScriptingFloatChannel, default=0, const [0,2]}
       CTRL_L_mouth_cornerSharpnessU {type=MovieSceneScriptingFloatChannel, default=0, keys [[296,-0.0077],[316,-0.0077],[350,0]]}
       CTRL_R_mouth_lipsTowardsTeethD {type=MovieSceneScriptingFloatChannel, default=0, const [0,3]}
       CTRL_L_mouth_lipsTowardsTeethD {type=MovieSceneScriptingFloatChannel, default=0, const [0,3]}
       CTRL_R_eye_eyelidU.X {type=MovieSceneScriptingFloatChannel, default=0, const [0,3]}
       CTRL_R_eye_eyelidU.Y {type=MovieSceneScriptingFloatChannel, default=0, const [0,3]}
       CTRL_L_mouth_lipsTowardsTeethU {type=MovieSceneScriptingFloatChannel, default=0, const [0,3]}
       CTRL_R_eye_eyelidD.X {type=MovieSceneScriptingFloatChannel, default=0, const [0,2]}
       CTRL_R_eye_eyelidD.Y {type=MovieSceneScriptingFloatChannel, default=0, const [0,2]}
       CTRL_R_mouth_lipsTowardsTeethU {type=MovieSceneScriptingFloatChannel, default=0, const [0,3]}
       CTRL_faceGUIfollowHead {type=MovieSceneScriptingFloatChannel, default=0, const [0,1]}
       CTRL_eyesAimFollowHead {type=MovieSceneScriptingFloatChannel, default=0, const [0,1]}
       CTRL_C_nose.X {type=MovieSceneScriptingFloatChannel, default=0, const [0,2]}
       CTRL_C_nose.Y {type=MovieSceneScriptingFloatChannel, default=0, const [0,2]}
       CTRL_C_teeth_rotateU {type=MovieSceneScriptingFloatChannel, default=0, const [0,3]}
       CTRL_C_teeth_rotateD {type=MovieSceneScriptingFloatChannel, default=0, const [0,3]}
       CTRL_L_brow_inner {
        type=MovieSceneScriptingFloatChannel
        default=0
        thinned_from=17
        keys [[93,0],[97,-0.8108],[104,-0.2503],[109,0.3222],[135,0.0056],[143,0],[145,-0.2628],[148,-0.1873],[152,-0.0218],[158,-0.0042],[314,0],[324,0],[328,-0.8108],[335,-0.2503],[348,0]]
       }
       CTRL_R_brow_inner {
        type=MovieSceneScriptingFloatChannel
        default=0
        thinned_from=17
        keys [[93,0],[97,-0.8108],[104,-0.2503],[109,0.3222],[135,0.0056],[143,0],[145,-0.2628],[148,-0.1873],[152,-0.0218],[158,-0.0042],[314,0],[324,0],[328,-0.8108],[335,-0.2503],[348,0]]
       }
       CTRL_L_brow_outer {
        type=MovieSceneScriptingFloatChannel
        default=0
        thinned_from=17
        keys [[93,0],[97,-1],[104,-0.3087],[109,0.3222],[135,0.0056],[143,0],[145,-0.3241],[148,-0.231],[152,-0.0269],[158,-0.0052],[314,0],[324,0],[328,-1],[335,-0.3087],[348,0]]
       }
       CTRL_R_brow_outer {
        type=MovieSceneScriptingFloatChannel
        default=0
        thinned_from=17
        keys [[93,0],[97,-1],[104,-0.3087],[109,0.3222],[135,0.0056],[143,0],[145,-0.3241],[148,-0.231],[152,-0.0269],[158,-0.0052],[314,0],[324,0],[328,-1],[335,-0.3087],[348,0]]
       }
       CTRL_C_mouth.X {type=MovieSceneScriptingFloatChannel, default=0, keys [[296,-0.0176],[316,-0.0176],[350,0]]}
       CTRL_C_mouth.Y {type=MovieSceneScriptingFloatChannel, default=0, keys [[296,-0.0234],[316,-0.0234],[350,0]]}
       CTRL_L_mouth_upperLipRaise {type=MovieSceneScriptingFloatChannel, default=0.0039, keys [[296,-0.0038],[316,-0.0038],[350,0.0039]]}
       CTRL_L_mouth_cornerPull {type=MovieSceneScriptingFloatChannel, default=0.148, thinned_from=4, keys [[124,0.0975],[316,0.0975],[350,0.148]]}
       CTRL_L_mouth_dimple {type=MovieSceneScriptingFloatChannel, default=0.0705, keys [[296,-0.0609],[316,-0.0609],[350,0.0705]]}
       CTRL_L_mouth_cornerDepress {type=MovieSceneScriptingFloatChannel, default=0.0006, keys [[296,-0.0014],[316,-0.0014],[350,0.0006]]}
       CTRL_L_mouth_stretch {type=MovieSceneScriptingFloatChannel, default=0, keys [[296,-0.0158],[316,-0.0158],[350,0]]}
       CTRL_L_mouth_lowerLipDepress {type=MovieSceneScriptingFloatChannel, default=0.0005, keys [[296,-0.0003],[316,-0.0003],[350,0.0005]]}
       CTRL_R_mouth_lowerLipDepress {type=MovieSceneScriptingFloatChannel, default=0.0005, keys [[296,-0.0005],[316,-0.0005],[350,0.0005]]}
       CTRL_R_mouth_stretch {type=MovieSceneScriptingFloatChannel, default=0, keys [[296,-0.0124],[316,-0.0124],[350,0]]}
       CTRL_L_mouth_towardsD {type=MovieSceneScriptingFloatChannel, default=0.0004, keys [[296,-0.0013],[316,-0.0013],[350,0.0004]]}
       CTRL_R_mouth_funnelU {type=MovieSceneScriptingFloatChannel, default=0.0025, keys [[296,-0.0039],[316,-0.0039],[350,0.0025]]}
       CTRL_L_mouth_towardsU {type=MovieSceneScriptingFloatChannel, default=0.0007, keys [[296,-0.0052],[316,-0.0052],[350,0.0007]]}
       CTRL_R_mouth_lipBiteD {type=MovieSceneScriptingFloatChannel, default=0.0075, keys [[296,-0.0055],[316,-0.0055],[350,0.0075]]}
       CTRL_R_mouth_funnelD {type=MovieSceneScriptingFloatChannel, default=0, keys [[296,-0.0008],[316,-0.0008],[350,0]]}
       CTRL_L_mouth_pressU {type=MovieSceneScriptingFloatChannel, default=0, const [0,3]}
       CTRL_L_mouth_funnelD {type=MovieSceneScriptingFloatChannel, default=0.0008, keys [[296,0.0001],[316,0.0001],[350,0.0008]]}
       CTRL_L_mouth_suckBlow {type=MovieSceneScriptingFloatChannel, default=0, keys [[296,-0.0027],[316,-0.0027],[350,0]]}
       CTRL_L_mouth_pressD {type=MovieSceneScriptingFloatChannel, default=0, const [0,3]}
       CTRL_L_mouth_funnelU {type=MovieSceneScriptingFloatChannel, default=0.0019, keys [[296,-0.0031],[316,-0.0031],[350,0.0019]]}
       CTRL_R_mouth_towardsU {type=MovieSceneScriptingFloatChannel, default=0.0013, keys [[296,-0.0077],[316,-0.0077],[350,0.0013]]}
       CTRL_R_mouth_purseD {type=MovieSceneScriptingFloatChannel, default=0, keys [[296,-0.0117],[316,-0.0117],[350,0]]}
       CTRL_L_mouth_lipBiteU {type=MovieSceneScriptingFloatChannel, default=-0.0001, keys [[296,-0.0133],[316,-0.0133],[350,-0.0001]]}
       CTRL_R_mouth_pressD {type=MovieSceneScriptingFloatChannel, default=0, const [0,3]}
       CTRL_R_mouth_pressU {type=MovieSceneScriptingFloatChannel, default=0, const [0,3]}
       CTRL_L_mouth_purseD {type=MovieSceneScriptingFloatChannel, default=0, keys [[296,-0.0216],[316,-0.0216],[350,0]]}
       CTRL_R_mouth_lipBiteU {type=MovieSceneScriptingFloatChannel, default=0, keys [[296,-0.011],[316,-0.011],[350,0]]}
       CTRL_R_mouth_suckBlow {type=MovieSceneScriptingFloatChannel, default=0, keys [[296,-0.0091],[316,-0.0091],[350,0]]}
       CTRL_L_mouth_purseU {type=MovieSceneScriptingFloatChannel, default=0, keys [[296,-0.0283],[316,-0.0283],[350,0]]}
       CTRL_R_mouth_towardsD {type=MovieSceneScriptingFloatChannel, default=0.0012, keys [[296,-0.0011],[316,-0.0011],[350,0.0012]]}
       CTRL_L_mouth_lipBiteD {type=MovieSceneScriptingFloatChannel, default=0.0078, keys [[296,-0.0055],[316,-0.0055],[350,0.0078]]}
       CTRL_R_mouth_purseU {type=MovieSceneScriptingFloatChannel, default=0, keys [[296,-0.0525],[316,-0.0525],[350,0]]}
       CTRL_R_mouth_cornerDepress {type=MovieSceneScriptingFloatChannel, default=0, keys [[296,-0.0021],[316,-0.0021],[350,0]]}
       CTRL_R_mouth_upperLipRaise {type=MovieSceneScriptingFloatChannel, default=0.0007, keys [[296,-0.0003],[316,-0.0003],[350,0.0007]]}
       CTRL_R_mouth_dimple {type=MovieSceneScriptingFloatChannel, default=0.098, keys [[296,0.0043],[316,0.0043],[350,0.0416]]}
       CTRL_R_mouth_cornerPull {type=MovieSceneScriptingFloatChannel, default=0.0305, keys [[296,-0.1406],[316,-0.1406],[350,0.0305]]}
      ]
     }
    ]
   }
   {
    type=MovieSceneControlRigParameterTrack
    display_name=Fortnite_Biped_CtrlRig
    sections [
     {
      type=MovieSceneControlRigParameterSection
      channels [
       leg_l_stretch_switch {
        type=MovieSceneScriptingBoolChannel
        default=false
        keys [[0,false],[92,false],[94,false],[96,false],[98,false],[100,false],[102,false],[104,false],[106,false],[108,false],[110,false],[112,false],[114,false],[116,false],[118,false],[120,false],[122,false],[124,false],[126,false],[128,false],[130,false],[132,false],[134,false],[136,false],[138,false],[140,false],[142,false],[144,false],[146,false],[148,false],[150,false],[152,false],[154,false],[156,false],[158,false],[160,false],[162,false],[164,false],[166,false],[168,false],[170,false],[172,false],[174,false],[176,false],[178,false],[180,false],[182,false],[184,false],[186,false],[188,false],[190,false],[192,false],[194,false],[196,false],[198,false],[200,false],[202,false],[204,false],[206,false],[208,false],[210,false],[212,false],[214,false],[216,false],[218,false],[220,false],[222,false],[224,false],[226,false],[228,false],[230,false],[232,false],[234,false],[236,false],[238,false],[240,false],[242,false],[244,false],[246,false],[248,false],[250,false],[252,false],[254,false],[256,false],[258,false],[260,false],[262,false],[264,false],[266,false],[268,false],[270,false],[272,false],[274,false],[276,false],[278,false],[280,false],[282,false],[284,false],[286,false],[288,false],[290,false],[292,false],[294,false],[296,false],[298,false],[300,false],[302,false],[304,false],[306,false],[308,false],[310,false],[312,false],[314,false],[316,false],[318,false],[320,false],[322,false],[324,false],[326,false],[328,false],[330,false],[332,false],[334,false],[336,false],[338,false],[340,false],[342,false],[344,false],[346,true],[348,false],[350,true],[352,false],[354,false],[356,true],[358,false],[360,true],[362,true],[364,true],[366,false],[368,false]]
       }
       leg_r_stretch_switch {
        type=MovieSceneScriptingBoolChannel
        default=false
        keys [[0,false],[92,false],[94,false],[96,false],[98,false],[100,false],[102,false],[104,false],[106,false],[108,false],[110,false],[112,false],[114,false],[116,false],[118,false],[120,false],[122,false],[124,false],[126,false],[128,false],[130,false],[132,false],[134,false],[136,false],[138,false],[140,false],[142,false],[144,false],[146,false],[148,false],[150,false],[152,false],[154,false],[156,false],[158,false],[160,false],[162,false],[164,false],[166,false],[168,false],[170,false],[172,false],[174,false],[176,false],[178,false],[180,false],[182,false],[184,false],[186,false],[188,false],[190,false],[192,false],[194,false],[196,false],[198,false],[200,false],[202,false],[204,false],[206,false],[208,false],[210,false],[212,false],[214,false],[216,false],[218,false],[220,false],[222,false],[224,false],[226,false],[228,false],[230,false],[232,false],[234,false],[236,false],[238,false],[240,false],[242,false],[244,false],[246,false],[248,false],[250,false],[252,false],[254,false],[256,false],[258,false],[260,false],[262,false],[264,false],[266,false],[268,false],[270,false],[272,false],[274,false],[276,false],[278,false],[280,false],[282,false],[284,false],[286,false],[288,false],[290,false],[292,false],[294,false],[296,false],[298,false],[300,false],[302,false],[304,false],[306,false],[308,false],[310,false],[312,false],[314,false],[316,false],[318,false],[320,false],[322,false],[324,false],[326,false],[328,false],[330,false],[332,false],[334,false],[336,false],[338,false],[340,false],[342,false],[344,false],[346,true],[348,false],[350,true],[352,false],[354,false],[356,true],[358,false],[360,true],[362,true],[364,true],[366,false],[368,false]]
       }
       arm_l_fk_local {type=MovieSceneScriptingBoolChannel, default=true, const [true,72]}
       Auto Interpolate_3 {
        type=MovieSceneScriptingBoolChannel
        default=false
        keys [[0,true],[92,true],[330,true],[340,false],[342,false],[344,false],[346,true],[348,false],[350,true],[352,false],[354,false],[356,true],[358,false],[360,true],[362,true],[364,true],[366,false],[368,false]]
       }
       Auto Interpolate {
        type=MovieSceneScriptingBoolChannel
        default=false
        keys [[0,true],[178,true],[194,true],[212,true],[250,true],[264,true],[300,true],[304,true],[330,true],[336,true],[340,false],[342,false],[344,false],[346,true],[348,false],[350,true],[352,false],[354,false],[356,true],[358,false],[360,true],[362,true],[364,true],[366,false],[368,false]]
       }
       arm_l_stretch_switch {
        type=MovieSceneScriptingBoolChannel
        default=false
        keys [[0,false],[92,false],[94,false],[96,false],[98,false],[100,false],[102,false],[104,false],[106,false],[108,false],[110,false],[112,false],[114,false],[116,false],[118,false],[120,false],[122,false],[124,false],[126,false],[128,false],[130,false],[132,false],[134,false],[136,false],[138,false],[140,false],[142,false],[144,false],[146,false],[148,false],[150,false],[152,false],[154,false],[156,false],[158,false],[160,false],[162,false],[164,false],[166,false],[168,false],[170,false],[172,false],[174,false],[176,false],[178,false],[180,false],[182,false],[184,false],[186,false],[188,false],[190,false],[192,false],[194,false],[196,false],[198,false],[200,false],[202,false],[204,false],[206,false],[208,false],[210,false],[212,false],[214,false],[216,false],[218,false],[220,false],[222,false],[224,false],[226,false],[228,false],[230,false],[232,false],[234,false],[236,false],[238,false],[240,false],[242,false],[244,false],[246,false],[248,false],[250,false],[252,false],[254,false],[256,false],[258,false],[260,false],[262,false],[264,false],[266,false],[268,false],[270,false],[272,false],[274,false],[276,false],[278,false],[280,false],[282,false],[284,false],[286,false],[288,false],[290,false],[292,false],[294,false],[296,false],[298,false],[300,false],[302,false],[304,false],[306,false],[308,false],[310,false],[312,false],[314,false],[316,false],[318,false],[320,false],[322,false],[324,false],[326,false],[328,false],[330,false],[332,false],[334,false],[336,false],[338,false],[340,false],[342,false],[344,false],[346,true],[348,false],[350,true],[352,false],[354,false],[356,true],[358,false],[360,true],[362,true],[364,true],[366,false],[368,false]]
       }
       Keep Aligned {
        type=MovieSceneScriptingBoolChannel
        default=false
        keys [[0,false],[92,false],[94,false],[96,false],[98,false],[100,false],[102,false],[104,false],[106,false],[108,false],[110,false],[112,false],[114,false],[116,false],[118,false],[120,false],[122,false],[124,false],[126,false],[128,false],[130,false],[132,false],[134,false],[136,false],[138,false],[140,false],[142,false],[144,false],[146,false],[148,false],[150,false],[152,false],[154,false],[156,false],[158,false],[160,false],[162,false],[164,false],[166,false],[168,false],[170,false],[172,false],[174,false],[176,false],[178,false],[180,false],[182,false],[184,false],[186,false],[188,false],[190,false],[192,false],[194,false],[196,false],[198,false],[200,false],[202,false],[204,false],[206,false],[208,false],[210,false],[212,false],[214,false],[216,false],[218,false],[220,false],[222,false],[224,false],[226,false],[228,false],[230,false],[232,false],[234,false],[236,false],[238,false],[240,false],[242,false],[244,false],[246,false],[248,false],[250,false],[252,false],[254,false],[256,false],[258,false],[260,false],[262,false],[264,false],[266,false],[268,false],[270,false],[272,false],[274,false],[276,false],[278,false],[280,false],[282,false],[284,false],[286,false],[288,false],[290,false],[292,false],[294,false],[296,false],[298,false],[300,false],[302,false],[304,false],[306,false],[308,false],[310,false],[312,false],[314,false],[316,false],[318,false],[320,false],[322,false],[324,false],[326,false],[328,false],[330,false],[332,false],[334,false],[336,false],[338,false],[340,false],[342,false],[344,false],[346,true],[348,false],[350,true],[352,false],[354,false],[356,true],[358,false],[360,true],[362,true],[364,true],[366,false],[368,false]]
       }
       arm_r_fk_local {type=MovieSceneScriptingBoolChannel, default=true, const [true,95]}
       Auto Interpolate_4 {
        type=MovieSceneScriptingBoolChannel
        default=false
        keys [[0,true],[178,true],[194,true],[212,true],[222,true],[244,false],[246,false],[248,false],[250,false],[252,false],[254,false],[256,false],[258,false],[260,false],[262,false],[264,false],[266,false],[268,false],[270,false],[272,false],[274,false],[276,false],[278,false],[280,false],[282,false],[284,false],[286,false],[288,false],[290,false],[292,false],[294,false],[296,false],[298,false],[300,false],[302,false],[304,false],[306,false],[308,false],[310,false],[312,false],[314,false],[316,false],[318,false],[320,false],[322,false],[324,false],[326,false],[328,false],[330,false],[332,false],[334,false],[336,false],[338,false],[340,false],[342,false],[344,false],[346,true],[348,false],[350,true],[352,false],[354,false],[356,true],[358,false],[360,true],[362,true],[364,true],[366,false],[368,false]]
       }
       Auto Interpolate_2 {
        type=MovieSceneScriptingBoolChannel
        default=false
        keys [[0,true],[178,true],[194,true],[212,true],[222,true],[244,false],[246,false],[248,false],[250,false],[252,false],[254,false],[256,false],[258,false],[260,false],[262,false],[264,false],[266,false],[268,false],[270,false],[272,false],[274,false],[276,false],[278,false],[280,false],[282,false],[284,false],[286,false],[288,false],[290,false],[292,false],[294,false],[296,false],[298,false],[300,false],[302,false],[304,false],[306,false],[308,false],[310,false],[312,false],[314,false],[316,false],[318,false],[320,false],[322,false],[324,false],[326,false],[328,false],[330,false],[332,false],[334,false],[336,false],[338,false],[340,false],[342,false],[344,false],[346,true],[348,false],[350,true],[352,false],[354,false],[356,true],[358,false],[360,true],[362,true],[364,true],[366,false],[368,false]]
       }
       arm_r_stretch_switch {
        type=MovieSceneScriptingBoolChannel
        default=false
        keys [[0,false],[92,false],[94,false],[96,false],[98,false],[100,false],[102,false],[104,false],[106,false],[108,false],[110,false],[112,false],[114,false],[116,false],[118,false],[120,false],[122,false],[124,false],[126,false],[128,false],[130,false],[132,false],[134,false],[136,false],[138,false],[140,false],[142,false],[144,false],[146,false],[148,false],[150,false],[152,false],[154,false],[156,false],[158,false],[160,false],[162,false],[164,false],[166,false],[168,false],[170,false],[172,false],[174,false],[176,false],[178,false],[180,false],[182,false],[184,false],[186,false],[188,false],[190,false],[192,false],[194,false],[212,false],[214,false],[216,false],[218,false],[220,false],[222,false],[224,false],[226,false],[228,false],[230,false],[232,false],[234,false],[236,false],[238,false],[240,false],[242,false],[244,false],[246,false],[248,false],[250,false],[252,false],[254,false],[256,false],[258,false],[260,false],[262,false],[264,false],[266,false],[268,false],[270,false],[272,false],[274,false],[276,false],[278,false],[280,false],[282,false],[284,false],[286,false],[288,false],[290,false],[292,false],[294,false],[296,false],[298,false],[300,false],[302,false],[304,false],[306,false],[308,false],[310,false],[312,false],[314,false],[316,false],[318,false],[320,false],[322,false],[324,false],[326,false],[328,false],[330,false],[332,false],[334,false],[336,false],[338,false],[340,false],[342,false],[344,false],[346,true],[348,false],[350,true],[352,false],[354,false],[356,true],[358,false],[360,true],[362,true],[364,true],[366,false],[368,false]]
       }
       Keep Aligned_2 {
        type=MovieSceneScriptingBoolChannel
        default=false
        keys [[0,false],[92,false],[94,false],[96,false],[98,false],[100,false],[102,false],[104,false],[106,false],[108,false],[110,false],[112,false],[114,false],[116,false],[118,false],[120,false],[122,false],[124,false],[126,false],[128,false],[130,false],[132,false],[134,false],[136,false],[138,false],[140,false],[142,false],[144,false],[146,false],[148,false],[150,false],[152,false],[154,false],[156,false],[158,false],[160,false],[162,false],[164,false],[166,false],[168,false],[170,false],[172,false],[174,false],[176,false],[178,false],[180,false],[182,false],[184,false],[186,false],[188,false],[190,false],[192,false],[194,false],[212,false],[214,false],[216,false],[218,false],[220,false],[222,false],[224,false],[226,false],[228,false],[230,false],[232,false],[234,false],[236,false],[238,false],[240,false],[242,false],[244,false],[246,false],[248,false],[250,false],[252,false],[254,false],[256,false],[258,false],[260,false],[262,false],[264,false],[266,false],[268,false],[270,false],[272,false],[274,false],[276,false],[278,false],[280,false],[282,false],[284,false],[286,false],[288,false],[290,false],[292,false],[294,false],[296,false],[298,false],[300,false],[302,false],[304,false],[306,false],[308,false],[310,false],[312,false],[314,false],[316,false],[318,false],[320,false],[322,false],[324,false],[326,false],[328,false],[330,false],[332,false],[334,false],[336,false],[338,false],[340,false],[342,false],[344,false],[346,true],[348,false],[350,true],[352,false],[354,false],[356,true],[358,false],[360,true],[362,true],[364,true],[366,false],[368,false]]
       }
       head_local {type=MovieSceneScriptingBoolChannel, default=true, const [true,116]}
       Stretch {type=MovieSceneScriptingBoolChannel, default=false, const [false,109]}
       head_stretch {
        type=MovieSceneScriptingBoolChannel
        default=false
        keys [[0,false],[92,false],[94,false],[96,false],[98,false],[100,false],[102,false],[104,false],[106,false],[108,false],[110,false],[112,false],[114,false],[116,false],[118,false],[120,false],[122,false],[124,false],[126,false],[128,false],[130,false],[132,false],[134,false],[136,false],[138,false],[140,false],[142,false],[144,false],[146,false],[148,false],[150,false],[152,false],[154,false],[156,false],[158,false],[160,false],[162,false],[164,false],[166,false],[168,false],[170,false],[172,false],[174,false],[176,false],[178,false],[180,false],[182,false],[184,false],[186,false],[188,false],[190,false],[192,false],[194,false],[196,false],[198,false],[200,false],[202,false],[204,false],[206,false],[208,false],[210,false],[212,false],[214,false],[216,false],[218,false],[220,false],[222,false],[224,false],[226,false],[228,false],[230,false],[232,false],[234,false],[236,false],[238,false],[240,false],[242,false],[244,false],[246,false],[248,false],[250,false],[252,false],[254,false],[256,false],[258,false],[260,false],[262,false],[264,false],[266,false],[268,false],[270,false],[272,false],[274,false],[276,false],[278,false],[280,false],[282,false],[284,false],[286,false],[288,false],[290,false],[292,false],[294,false],[296,false],[298,false],[300,false],[302,false],[304,false],[306,false],[308,false],[310,false],[312,false],[314,false],[316,false],[318,false],[320,false],[322,false],[324,false],[326,false],[328,false],[330,false],[332,false],[334,false],[336,false],[338,false],[340,false],[342,false],[344,false],[346,true],[348,false],[350,true],[352,false],[354,false],[356,true],[358,false],[360,true],[362,true],[364,true],[366,false],[368,false]]
       }
       spine_stretch_switch {
        type=MovieSceneScriptingBoolChannel
        default=false
        keys [[0,false],[92,false],[94,false],[96,false],[98,false],[100,false],[102,false],[104,false],[106,false],[108,false],[110,false],[112,false],[114,false],[116,false],[118,false],[120,false],[122,false],[124,false],[126,false],[128,false],[130,false],[132,false],[134,false],[136,false],[138,false],[140,false],[142,false],[144,false],[146,false],[148,false],[150,false],[152,false],[154,false],[156,false],[158,false],[160,false],[162,false],[164,false],[166,false],[168,false],[170,false],[172,false],[174,false],[176,false],[178,false],[180,false],[182,false],[184,false],[186,false],[188,false],[190,false],[192,false],[194,false],[196,false],[198,false],[200,false],[202,false],[204,false],[206,false],[208,false],[210,false],[212,false],[214,false],[216,false],[218,false],[220,false],[222,false],[224,false],[226,false],[228,false],[230,false],[232,false],[234,false],[236,false],[238,false],[240,false],[242,false],[244,false],[246,false],[248,false],[250,false],[252,false],[254,false],[256,false],[258,false],[260,false],[262,false],[264,false],[266,false],[268,false],[270,false],[272,false],[274,false],[276,false],[278,false],[280,false],[282,false],[284,false],[286,false],[288,false],[290,false],[292,false],[294,false],[296,false],[298,false],[300,false],[302,false],[304,false],[306,false],[308,false],[310,false],[312,false],[314,false],[316,false],[318,false],[320,false],[322,false],[324,false],[326,false],[328,false],[330,false],[332,false],[334,false],[336,false],[338,false],[340,false],[342,false],[344,false],[346,true],[348,false],[350,true],[352,false],[354,false],[356,true],[358,false],[360,true],[362,true],[364,true],[366,false],[368,false]]
       }
       arm_l_fk_ik_switch {type=MovieSceneScriptingBoolChannel, default=false, const [false,140]}
       leg_l_fk_ik_switch {type=MovieSceneScriptingBoolChannel, default=false, const [false,140]}
       arm_r_fk_ik_switch {type=MovieSceneScriptingBoolChannel, default=false, const [false,140]}
       leg_r_fk_ik_switch {type=MovieSceneScriptingBoolChannel, default=false, const [false,140]}
       ShowBodyControls {type=MovieSceneScriptingBoolChannel, default=true, const [true,140]}
       spine_fk_ik_switch {type=MovieSceneScriptingBoolChannel, default=false, const [false,140]}
       neck_fk_ik_switch {type=MovieSceneScriptingBoolChannel, default=false, const [false,140]}
       space {type=MovieSceneScriptingByteChannel, default=1, const [1,115]}
       global_ctrl.Location.X {type=MovieSceneScriptingFloatChannel, default=0, const [0,140]}
       global_ctrl.Location.Y {type=MovieSceneScriptingFloatChannel, default=0, const [0,140]}
       global_ctrl.Location.Z {type=MovieSceneScriptingFloatChannel, default=0, const [0,140]}
       global_ctrl.Rotation.X {type=MovieSceneScriptingFloatChannel, default=0, const [0,140]}
       global_ctrl.Rotation.Y {type=MovieSceneScriptingFloatChannel, default=0, const [0,140]}
       global_ctrl.Rotation.Z {type=MovieSceneScriptingFloatChannel, default=0, const [0,140]}
       root_ctrl.Location.X {
        type=MovieSceneScriptingFloatChannel
        default=0
        thinned_from=371
        keys [[-1,18.7997],[92,18.7997],[98,18.5104],[103,18.2155],[104,18.1268],[106,17.8532],[108,17.7023],[110,17.4242],[116,16.8292],[119,16.4462],[121,16.2588],[122,16.197],[128,16.0031],[134,15.9293],[141,16.0395],[148,16.3028],[188,16.2032],[215,16.6385],[241,17.1784],[270,17.3523],[282,17.6095],[306,17.5823],[334,16.9266],[336,16.7703],[338,16.5303],[340,16.1794],[342,15.7713],[344,15.5292],[348,14.8506],[350,14.4033],[351,14.1412],[352,13.9634],[354,13.8923],[355,13.9191],[356,13.9779],[357,14.1375],[358,14.4375],[359,14.9634],[360,15.7266],[361,16.9259],[362,17.7155],[363,17.4335],[364,16.813],[365,14.8337],[366,11.0208],[367,-0.0459],[368,-8.9405],[369,-8.9405]]
       }
       root_ctrl.Location.Y {
        type=MovieSceneScriptingFloatChannel
        default=0
        thinned_from=371
        keys [[-1,-0.5735],[208,-0.5735],[209,-0.5613],[210,-0.5432],[211,-0.5327],[212,-0.5265],[215,-0.5199],[216,-0.5189],[218,-0.5208],[220,-0.5278],[221,-0.5257],[222,-0.5212],[223,-0.5153],[224,-0.5061],[225,-0.4909],[227,-0.4528],[228,-0.4277],[229,-0.3866],[230,-0.3439],[231,-0.3143],[232,-0.2871],[234,-0.2228],[235,-0.1952],[236,-0.1699],[237,-0.1456],[238,-0.1243],[239,-0.1078],[240,-0.0952],[241,-0.0873],[242,-0.0804],[244,-0.0589],[245,-0.0513],[246,-0.0447],[247,-0.0378],[248,-0.0314],[249,-0.0258],[250,-0.0207],[251,-0.0164],[252,-0.0127],[253,-0.0099],[254,-0.0067],[255,-0.0017],[256,0.0036],[257,0.0078],[258,0.0116],[259,0.0158],[260,0.0185],[261,0.0191],[262,0.0193],[263,0.0154],[264,0.0088],[265,0.0041],[266,-0.0018],[267,-0.0172],[268,-0.0365],[270,-0.0642],[271,-0.081],[273,-0.1172],[274,-0.1365],[276,-0.1811],[277,-0.2011],[278,-0.2199],[279,-0.2381],[280,-0.2551],[281,-0.2699],[282,-0.2832],[283,-0.2956],[284,-0.3061],[285,-0.3133],[286,-0.3183],[287,-0.3223],[288,-0.3241],[289,-0.3219],[292,-0.3115],[293,-0.3091],[294,-0.308],[296,-0.3079],[299,-0.3098],[302,-0.3094],[303,-0.3084],[304,-0.3066],[305,-0.2998],[306,-0.2864],[307,-0.265],[308,-0.2398],[309,-0.218],[310,-0.1947],[311,-0.1671],[312,-0.1362],[313,-0.102],[314,-0.0649],[315,-0.0238],[316,0.0167],[317,0.052],[318,0.0839],[319,0.1154],[320,0.1375],[321,0.1457],[322,0.1525],[323,0.1633],[324,0.1708],[325,0.1725],[326,0.1732],[327,0.1722],[328,0.1699],[329,0.1603],[330,0.1411],[331,0.1119],[332,0.0649],[333,-0.0209],[334,-0.1641],[335,-0.4543],[336,-0.8637],[337,-1.3205],[338,-1.8041],[339,-2.2789],[340,-2.7183],[341,-3.0929],[342,-3.3754],[343,-3.5341],[344,-3.6536],[345,-3.7862],[346,-3.8967],[347,-3.9673],[348,-4.0117],[349,-4.0428],[350,-4.057],[351,-4.032],[352,-3.977],[353,-3.8935],[354,-3.8008],[355,-3.7243],[356,-3.663],[359,-3.5654],[360,-3.5205],[361,-3.4502],[362,-3.3423],[363,-0.5259],[364,5.5633],[365,18.619],[366,41.9983],[367,101.1137],[368,146.5659],[369,146.5659]]
       }
       root_ctrl.Location.Z {
        type=MovieSceneScriptingFloatChannel
        default=0
        thinned_from=371
        keys [[-1,-25.128],[92,-25.128],[96,-25.3],[98,-25.2833],[102,-24.9642],[103,-24.8075],[104,-24.5567],[105,-24.095],[106,-23.571],[107,-23.2304],[108,-22.8365],[110,-21.4252],[112,-20.4079],[113,-19.7574],[115,-18.34],[116,-17.5919],[119,-15.1411],[121,-13.5571],[122,-12.8043],[123,-12.1612],[124,-11.5474],[126,-10.2336],[127,-9.662],[128,-9.1466],[129,-8.6859],[130,-8.2834],[131,-7.9391],[132,-7.6625],[133,-7.4443],[134,-7.3398],[136,-7.3287],[137,-7.3787],[138,-7.4614],[142,-7.7018],[146,-7.8783],[149,-8.1368],[152,-8.4435],[155,-8.8181],[162,-9.8367],[184,-13.6601],[189,-14.3695],[193,-14.8068],[198,-15.1352],[203,-15.3038],[206,-15.3107],[210,-15.1646],[216,-14.7634],[222,-14.1473],[230,-13.1067],[248,-10.3341],[253,-9.6642],[260,-8.8819],[263,-8.6034],[266,-8.3821],[270,-8.1611],[272,-8.083],[274,-8.0477],[279,-8.0829],[292,-8.4405],[299,-8.7471],[311,-9.4685],[318,-9.7159],[327,-9.8932],[332,-9.9532],[333,-9.9891],[334,-10.0496],[335,-10.2812],[336,-10.7328],[337,-11.4043],[338,-12.2991],[339,-13.3847],[340,-14.5679],[341,-15.7858],[342,-16.9143],[343,-17.7115],[344,-18.4685],[346,-20.4936],[347,-21.3661],[349,-22.9348],[350,-23.6128],[351,-24.1472],[352,-24.5604],[353,-24.8465],[355,-25.1866],[356,-25.2548],[358,-25.2103],[359,-25.4159],[360,-25.8681],[361,-26.6503],[362,-27.1859],[363,-26.0032],[364,-23.4013],[365,-14.9151],[366,1.4835],[367,36.6275],[368,62.4397],[369,62.4397]]
       }
       root_ctrl.Rotation.X {
        type=MovieSceneScriptingFloatChannel
        default=0
        thinned_from=371
        keys [[-1,116.2705],[92,116.2705],[95,115.2682],[98,113.5827],[102,110.6239],[103,109.2938],[104,107.3897],[105,104.6232],[106,101.6165],[107,99.1539],[108,96.4612],[110,88.8468],[112,82.6104],[114,75.7834],[116,69.3447],[118,62.6604],[119,59.6361],[121,53.8927],[122,51.3457],[123,49.6401],[124,48.0921],[126,43.9286],[127,41.9715],[128,40.2095],[130,37.6064],[132,34.4646],[135,30.4444],[136,29.1481],[139,25.5884],[140,24.4543],[142,22.3965],[145,18.988],[146,17.9212],[147,16.9525],[148,16.0416],[151,13.4756],[152,12.6682],[153,11.9182],[154,11.2302],[155,10.6263],[156,10.0759],[158,9.0664],[159,8.5006],[160,7.9593],[161,7.5204],[162,7.146],[163,6.8283],[164,6.5554],[165,6.3178],[168,5.6719],[170,5.2212],[171,4.9873],[172,4.7683],[173,4.5938],[175,4.2742],[176,4.1229],[177,3.9956],[179,3.7665],[180,3.6594],[181,3.56],[182,3.4679],[184,3.306],[186,3.0978],[187,3.016],[190,2.7839],[191,2.7188],[193,2.6064],[194,2.558],[195,2.515],[196,2.4797],[197,2.4511],[198,2.4376],[199,2.4472],[200,2.4619],[204,2.4888],[206,2.5762],[207,2.5426],[208,2.4686],[209,2.3634],[210,2.2487],[211,2.148],[212,2.0706],[214,2],[215,1.9336],[217,1.7855],[218,1.7195],[219,1.6661],[220,1.6348],[221,1.6238],[222,1.6189],[223,1.6329],[225,1.6805],[226,1.6965],[228,1.7018],[230,1.6944],[231,1.6798],[232,1.6559],[236,1.5418],[237,1.5185],[238,1.5062],[240,1.5047],[242,1.499],[244,1.509],[248,1.5091],[249,1.5151],[250,1.5264],[251,1.5411],[252,1.5512],[254,1.553],[256,1.5332],[259,1.5228],[260,1.5216],[261,1.5289],[262,1.545],[263,1.5764],[264,1.6145],[265,1.6389],[266,1.668],[267,1.7368],[268,1.817],[269,1.8831],[270,1.92],[272,1.9261],[273,1.9033],[274,1.8606],[275,1.808],[276,1.7666],[277,1.7503],[278,1.7429],[279,1.7606],[281,1.8305],[282,1.8743],[284,1.9905],[285,2.0403],[286,2.0848],[287,2.1241],[288,2.159],[290,2.2179],[291,2.2518],[292,2.2906],[293,2.3427],[294,2.3884],[296,2.4277],[297,2.4722],[298,2.5221],[299,2.5582],[301,2.6176],[303,2.6599],[304,2.6687],[306,2.6556],[307,2.6215],[308,2.5714],[312,2.4044],[315,2.2555],[316,2.2187],[317,2.1959],[318,2.1855],[319,2.1972],[321,2.2378],[322,2.264],[323,2.3359],[324,2.4452],[325,2.5608],[326,2.6984],[327,2.8657],[328,3.0547],[329,3.2539],[330,3.4812],[331,3.7258],[332,4.0624],[333,4.5658],[334,5.3245],[335,6.8662],[336,8.8491],[338,12.6401],[339,14.494],[340,16.214],[341,17.8288],[342,18.9337],[343,19.3386],[344,19.6717],[345,20.1638],[346,20.5725],[348,20.8965],[349,21.1995],[350,21.5544],[351,21.9547],[352,22.283],[358,23.2421],[359,23.4817],[360,23.6705],[362,23.7945],[363,26.3574],[364,31.5663],[365,37.9874],[366,45.3338],[367,54.7297],[368,60.5079],[369,60.5079]]
       }
       root_ctrl.Rotation.Y {
        type=MovieSceneScriptingFloatChannel
        default=0
        thinned_from=371
        keys [[-1,-8.8844],[92,-8.8844],[94,-8.9114],[95,-8.8885],[97,-8.7724],[98,-8.6895],[100,-8.4813],[102,-8.3509],[103,-8.2531],[104,-8.1118],[105,-7.8216],[106,-7.4717],[107,-7.2086],[108,-6.9164],[110,-6.0435],[111,-5.7524],[112,-5.4876],[114,-4.8245],[115,-4.5449],[116,-4.2763],[118,-3.7008],[120,-3.169],[121,-2.8516],[122,-2.5709],[124,-2.2923],[125,-2.0282],[126,-1.7284],[127,-1.4323],[128,-1.1716],[130,-0.8908],[131,-0.624],[132,-0.3411],[134,0.0679],[135,0.3419],[136,0.6172],[137,0.8464],[138,1.0371],[139,1.1867],[140,1.2787],[141,1.3245],[142,1.3454],[143,1.2912],[144,1.1744],[145,1.0128],[146,0.8441],[147,0.7246],[148,0.6153],[149,0.497],[150,0.3662],[151,0.2036],[152,0.0373],[153,-0.1129],[154,-0.2408],[155,-0.3239],[156,-0.391],[157,-0.4633],[158,-0.5413],[159,-0.6329],[160,-0.7372],[161,-0.8647],[162,-0.989],[163,-1.0809],[165,-1.2367],[166,-1.3043],[167,-1.3601],[168,-1.396],[169,-1.4139],[170,-1.422],[171,-1.4072],[174,-1.3287],[176,-1.301],[177,-1.2796],[178,-1.254],[180,-1.1934],[181,-1.1665],[182,-1.1361],[183,-1.0887],[184,-1.0473],[185,-1.0291],[186,-1.0145],[187,-0.9939],[188,-0.9786],[189,-0.9723],[190,-0.9695],[191,-0.9765],[192,-0.9921],[193,-1.0252],[194,-1.0759],[195,-1.1362],[196,-1.2002],[198,-1.314],[199,-1.376],[200,-1.4352],[203,-1.5712],[204,-1.625],[206,-1.7883],[207,-1.8335],[209,-1.8995],[211,-1.9577],[212,-1.983],[214,-2.018],[215,-2.0612],[219,-2.2775],[220,-2.3286],[221,-2.3721],[222,-2.4042],[223,-2.4241],[224,-2.4341],[227,-2.4426],[229,-2.4614],[230,-2.4754],[232,-2.5172],[234,-2.5389],[235,-2.5575],[236,-2.5813],[238,-2.6441],[240,-2.715],[242,-2.7569],[243,-2.8001],[245,-2.8994],[246,-2.9435],[247,-2.9769],[248,-3],[252,-3.0464],[254,-3.0587],[256,-3.1062],[259,-3.1578],[260,-3.1707],[264,-3.196],[268,-3.1958],[269,-3.1763],[272,-3.0688],[273,-3.0474],[274,-3.0376],[275,-3.0486],[278,-3.1103],[280,-3.1228],[282,-3.119],[284,-3.1499],[288,-3.1727],[289,-3.1659],[290,-3.151],[293,-3.0756],[294,-3.0593],[296,-3.0469],[299,-3.0555],[300,-3.0655],[301,-3.0853],[303,-3.1511],[304,-3.1948],[305,-3.245],[306,-3.3098],[308,-3.5059],[310,-3.6628],[312,-3.8298],[313,-3.8848],[316,-4.0294],[318,-4.0797],[320,-4.151],[322,-4.1813],[323,-4.2164],[324,-4.2404],[326,-4.232],[327,-4.1951],[328,-4.1385],[330,-4.0498],[331,-3.9965],[332,-3.965],[333,-3.9722],[334,-3.9882],[335,-4.1321],[336,-4.3964],[337,-4.6958],[338,-5.0209],[340,-5.6343],[341,-5.98],[342,-6.1807],[344,-6.1724],[345,-6.0776],[346,-5.9168],[347,-5.785],[348,-5.6198],[349,-5.3775],[350,-5.0971],[352,-4.4942],[353,-4.255],[354,-4.045],[355,-3.8447],[356,-3.6765],[357,-3.5713],[358,-3.483],[359,-3.3764],[360,-3.2495],[361,-3.1016],[362,-2.8939],[363,-1.5249],[364,0.3499],[365,1.3862],[366,2.2196],[367,3.2476],[368,3.8394],[369,3.8394]]
       }
       root_ctrl.Rotation.Z {
        type=MovieSceneScriptingFloatChannel
        default=0
        thinned_from=371
        keys [[-1,10.7206],[92,10.7206],[95,10.8079],[99,11.0488],[101,11.2478],[102,11.4056],[103,11.6859],[104,12.0735],[105,12.5452],[107,13.5778],[108,14.143],[110,15.5792],[113,17.4829],[115,18.6853],[116,19.2447],[117,19.7362],[118,20.1688],[119,20.54],[120,20.8536],[121,21.1065],[124,21.6627],[125,21.8982],[126,22.1783],[128,22.8811],[130,23.3094],[131,23.6154],[132,23.8331],[134,23.9146],[136,23.6455],[142,23.4394],[143,23.5153],[144,23.6823],[145,23.9753],[151,26.1402],[152,26.4687],[153,26.7357],[154,26.9439],[158,27.3647],[161,27.7956],[166,28.2912],[171,29.0406],[172,29.1365],[192,29.3782],[197,29.0605],[201,28.5723],[213,27.9136],[227,27.9441],[237,28.6491],[262,28.6777],[292,28.0184],[303,27.3173],[307,27.3507],[319,28.0079],[323,27.9984],[331,27.5841],[333,27.3822],[334,27.2327],[336,26.2056],[338,25.8453],[339,25.5387],[340,25.3403],[341,25.4469],[345,26.2661],[346,26.5041],[349,27.6162],[350,27.8734],[351,27.9847],[352,28.0353],[354,27.9047],[356,27.9368],[357,27.6771],[358,27.347],[360,27.1591],[361,27.0004],[362,26.764],[363,26.116],[364,24.9274],[365,22.4043],[366,18.7357],[367,13.5772],[368,10.253],[369,10.253]]
       }
       root_ctrl.Scale.X {type=MovieSceneScriptingFloatChannel, default=1, const [1,140]}
       root_ctrl.Scale.Y {type=MovieSceneScriptingFloatChannel, default=1, const [1,140]}
       root_ctrl.Scale.Z {type=MovieSceneScriptingFloatChannel, default=1, const [1,140]}
       body_offset_ctrl.Location.X {type=MovieSceneScriptingFloatChannel, default=0, const [0,140]}
       body_offset_ctrl.Location.Y {type=MovieSceneScriptingFloatChannel, default=0, const [0,140]}
       body_offset_ctrl.Location.Z {type=MovieSceneScriptingFloatChannel, default=0, const [0,140]}
       body_offset_ctrl.Rotation.X {type=MovieSceneScriptingFloatChannel, default=0, const [0,140]}
       body_offset_ctrl.Rotation.Y {type=MovieSceneScriptingFloatChannel, default=0, const [0,140]}
       body_offset_ctrl.Rotation.Z {type=MovieSceneScriptingFloatChannel, default=0, const [0,140]}
       body_offset_ctrl.Scale.X {type=MovieSceneScriptingFloatChannel, default=1, const [1,140]}
       body_offset_ctrl.Scale.Y {type=MovieSceneScriptingFloatChannel, default=1, const [1,140]}
       body_offset_ctrl.Scale.Z {type=MovieSceneScriptingFloatChannel, default=1, const [1,140]}
       thigh_l_fk_ctrl.Location.X {type=MovieSceneScriptingFloatChannel, default=0, thinned_from=134, keys [[0,0],[368,0]]}
       thigh_l_fk_ctrl.Location.Y {type=MovieSceneScriptingFloatChannel, default=0, thinned_from=134, keys [[0,0],[368,0]]}
       thigh_l_fk_ctrl.Location.Z {type=MovieSceneScriptingFloatChannel, default=0, thinned_from=134, keys [[0,0],[368,0]]}
       thigh_l_fk_ctrl.Rotation.X {
        type=MovieSceneScriptingFloatChannel
        default=0
        thinned_from=134
        keys [[0,-2.6414],[92,-2.6414],[94,-2.0215],[96,-1.2708],[98,-0.5531],[100,0],[102,0.322],[104,0.7877],[106,0.9762],[108,0.9137],[110,-0.1681],[112,-0.4437],[114,-0.9141],[116,-0.7859],[118,-0.2615],[120,0.209],[122,-0.0842],[124,0.184],[126,-0.0463],[128,0.0176],[130,0.5334],[132,1.4546],[134,2.5721],[136,3.2433],[138,4.1703],[140,5.5859],[142,6.4421],[144,7.8926],[146,8.9758],[148,9.7312],[150,10.6726],[152,11.4728],[154,11.9396],[156,12.2136],[160,12.6426],[162,12.7521],[164,13.0074],[166,13.3903],[172,13.9567],[174,13.7995],[176,13.5658],[178,12.9223],[180,12.3465],[182,12.0614],[184,12.0467],[186,12.1353],[190,12.3618],[192,12.7004],[194,13.2604],[196,13.6105],[198,13.6012],[200,13.6454],[204,13.5992],[212,13.269],[214,13.1132],[216,12.9841],[220,13.3642],[222,13.4837],[224,13.5234],[226,13.7354],[228,14.0706],[230,14.3346],[232,14.4728],[234,14.5567],[238,14.9268],[240,15.0169],[242,15.0641],[244,15.1505],[246,15.3856],[248,15.5143],[250,15.5642],[254,15.7492],[256,15.9112],[258,15.9177],[260,15.8555],[264,15.6616],[268,15.6741],[272,15.427],[274,15.4407],[276,15.5053],[278,15.6139],[280,15.7653],[282,15.7897],[284,15.6824],[286,15.54],[290,15.4205],[292,15.2767],[294,15.1773],[296,15.1648],[298,15.0826],[300,15.0611],[302,15.1368],[304,15.2596],[306,15.5394],[308,15.9111],[312,16.3968],[314,16.5198],[316,16.7498],[318,17.0168],[320,17.3762],[322,17.4484],[324,17.7063],[326,17.8525],[328,17.7204],[330,17.3107],[332,17.1232],[335,17.1845],[340,18.3684],[347,20.7212],[354,21.4056],[358,21.4734],[360,21.9841],[362,22.9174],[364,10.3055],[366,3.7544],[368,0]]
       }
       thigh_l_fk_ctrl.Rotation.Y {
        type=MovieSceneScriptingFloatChannel
        default=-7.0502
        thinned_from=134
        keys [[0,-7.8172],[92,-7.8172],[94,-7.4902],[96,-7.214],[98,-7.1306],[104,-7.1411],[106,-7.7861],[108,-8.3444],[110,-9.5427],[112,-10.072],[114,-10.812],[116,-11.2669],[118,-11.5634],[122,-11.7599],[124,-11.4924],[126,-11.6455],[128,-11.8745],[130,-11.6798],[132,-11.5442],[134,-11.2235],[136,-11.161],[138,-11.1864],[140,-11.017],[142,-10.5728],[144,-9.86],[146,-9.5097],[148,-9.2846],[152,-8.5777],[156,-8.1312],[158,-7.9301],[160,-7.772],[162,-7.6765],[164,-7.602],[166,-7.5042],[168,-7.3726],[170,-7.3448],[172,-7.514],[174,-7.639],[176,-7.6052],[178,-7.6458],[182,-7.8506],[184,-7.9113],[188,-7.8441],[190,-7.8266],[192,-7.7577],[196,-7.4432],[200,-7.3146],[202,-7.3109],[204,-7.3757],[206,-7.4783],[212,-7.4347],[214,-7.4374],[220,-7.2952],[222,-7.1806],[224,-7.1042],[226,-7.0421],[228,-7.0059],[230,-6.9975],[232,-6.8791],[234,-6.8027],[236,-6.7667],[242,-6.729],[244,-6.6993],[246,-6.6424],[248,-6.608],[250,-6.5488],[252,-6.4678],[256,-6.408],[258,-6.4206],[260,-6.4188],[262,-6.4557],[264,-6.5182],[266,-6.5449],[268,-6.6863],[270,-6.8907],[272,-7.0016],[274,-7.015],[276,-6.9942],[280,-6.8612],[282,-6.8506],[284,-6.9052],[286,-7.0027],[288,-7.1204],[292,-7.3924],[294,-7.4721],[298,-7.4989],[300,-7.4648],[306,-7.4548],[310,-7.4024],[312,-7.3517],[314,-7.2854],[318,-7.0893],[322,-6.7982],[326,-6.3978],[330,-6.0645],[332,-5.6751],[334,-5.1517],[335,-4.3931],[340,1.1028],[347,9.0984],[354,10.827],[356,10.8358],[358,10.8019],[360,9.2615],[362,6.2288],[364,-2.3528],[366,-4.12],[368,-5.797]]
       }
       thigh_l_fk_ctrl.Rotation.Z {
        type=MovieSceneScriptingFloatChannel
        default=0
        thinned_from=134
        keys [[0,12.7234],[92,12.7234],[94,12.9591],[96,13.483],[98,14.2679],[100,15.4143],[102,15.8871],[104,16.6917],[106,18.94],[108,19.5856],[110,21.8381],[112,22.5673],[114,23.7758],[116,24.7376],[118,26.1788],[122,28.9434],[124,28.6173],[126,29.6338],[128,30.3239],[130,29.7232],[132,29.2174],[134,27.8326],[136,26.5701],[138,25.5032],[140,24.5656],[142,23.5314],[144,23.1732],[146,23.0639],[148,22.5557],[150,21.7853],[152,21.1247],[158,19.359],[160,19.0368],[164,18.786],[166,18.6099],[168,18.2873],[170,18.0546],[172,17.9509],[176,17.4012],[180,16.5698],[184,15.78],[186,15.4982],[188,15.1591],[190,14.9825],[192,14.8559],[194,14.788],[196,14.7718],[200,14.8079],[204,14.7182],[212,14.7198],[216,14.5997],[222,14.7727],[224,14.7995],[228,14.9989],[230,15.1631],[232,15.1763],[236,15.4477],[238,15.539],[240,15.5972],[248,15.5964],[250,15.5705],[252,15.5074],[254,15.4854],[258,15.5579],[260,15.5517],[264,15.6164],[268,15.6243],[272,15.8081],[276,15.8675],[278,15.8383],[284,15.4682],[288,15.4799],[292,15.6322],[296,15.5963],[300,15.4661],[304,15.1627],[308,14.9674],[312,15.0317],[316,15.2294],[318,15.37],[320,15.4672],[322,15.4834],[326,15.3505],[328,15.2458],[330,15.0705],[332,14.5282],[335,13.6142],[340,3.3549],[347,-20.4469],[354,-28.0388],[356,-28.6842],[358,-28.7855],[360,-26.0985],[362,-20.3516],[364,-6.7676],[366,-0.6886],[368,0]]
       }
       thigh_l_fk_ctrl.Scale.X {type=MovieSceneScriptingFloatChannel, default=1, const [1,134]}
       thigh_l_fk_ctrl.Scale.Y {type=MovieSceneScriptingFloatChannel, default=1, const [1,134]}
       thigh_l_fk_ctrl.Scale.Z {type=MovieSceneScriptingFloatChannel, default=1, const [1,134]}
       calf_l_fk_ctrl.Location.X {
        type=MovieSceneScriptingFloatChannel
        default=0
        thinned_from=132
        keys [[0,0.0001],[92,0.0001],[94,-0.0001],[104,0],[108,0.0006],[112,0.0018],[114,0.0015],[116,0.0018],[118,0.0014],[120,0],[122,0],[124,-0.0003],[126,0.0005],[128,-0.0002],[130,-0.0001],[318,-0.0001],[330,-0.0001],[332,0.0006],[368,0]]
       }
       calf_l_fk_ctrl.Location.Y {
        type=MovieSceneScriptingFloatChannel
        default=0
        thinned_from=132
        keys [[0,0.0019],[92,0.0019],[94,-0.0005],[96,0.0002],[98,0.0002],[100,0],[102,0.0002],[104,-0.0021],[106,-0.0019],[108,-0.0001],[110,0.0043],[112,0.0168],[114,-0.0016],[116,0.0172],[118,-0.0041],[120,-0.0033],[122,-0.003],[124,-0.0034],[126,-0.0023],[128,-0.0023],[130,-0.0015],[132,-0.0025],[138,-0.0022],[140,-0.0018],[142,-0.0023],[144,-0.0024],[146,-0.0021],[148,-0.0023],[152,-0.0017],[154,-0.002],[158,-0.0014],[162,-0.0014],[166,-0.0023],[168,-0.0016],[170,-0.0018],[174,-0.0016],[176,-0.0013],[180,-0.0017],[186,-0.0015],[188,-0.0013],[190,-0.0018],[192,-0.0015],[194,-0.0015],[196,-0.0019],[198,-0.0013],[202,-0.0017],[204,-0.0011],[206,-0.0017],[220,-0.0013],[222,-0.0011],[226,-0.0016],[232,-0.0015],[234,-0.0011],[236,-0.0015],[238,-0.0016],[240,-0.0008],[246,-0.0016],[252,-0.0008],[254,-0.001],[258,-0.0009],[260,-0.0012],[262,-0.0008],[264,-0.0012],[268,-0.0004],[272,-0.0008],[278,-0.0007],[282,-0.0001],[284,-0.0005],[288,-0.0003],[290,-0.0008],[294,0],[296,-0.0006],[298,0],[300,-0.0002],[304,0],[306,-0.0003],[308,-0.0001],[312,-0.0002],[314,-0.0007],[318,-0.0007],[320,0.0004],[324,0.0001],[326,0.0003],[328,-0.0007],[330,-0.0004],[332,-0.0003],[340,0.0011],[347,0.0028],[354,0.0019],[356,0.0012],[358,0.0009],[360,0.0002],[362,0],[368,0]]
       }
       calf_l_fk_ctrl.Location.Z {
        type=MovieSceneScriptingFloatChannel
        default=0
        thinned_from=132
        keys [[0,0.0011],[92,0.0011],[94,-0.0005],[96,0.0001],[102,0],[104,-0.0017],[106,-0.0018],[108,-0.0009],[110,0.0002],[112,0.0037],[114,-0.0021],[116,0.0021],[118,-0.0021],[124,-0.0019],[126,-0.0023],[128,-0.0018],[162,-0.0014],[166,-0.0015],[168,-0.0014],[196,-0.0013],[204,-0.0012],[238,-0.0011],[240,-0.0009],[248,-0.0009],[262,-0.0006],[264,-0.0007],[268,-0.0005],[282,-0.0003],[294,-0.0002],[296,-0.0003],[304,-0.0001],[312,-0.0001],[318,-0.0003],[320,0],[324,-0.0001],[326,0.0001],[328,-0.0003],[347,0.0009],[354,0.0006],[362,0],[368,0]]
       }
       calf_l_fk_ctrl.Rotation.X {
        type=MovieSceneScriptingFloatChannel
        default=0
        thinned_from=132
        keys [[0,0.0118],[92,0.0118],[94,-0.0009],[96,-0.0098],[98,-0.0133],[100,0],[102,0.012],[104,0.0386],[106,0.1016],[108,0.1551],[110,0.2993],[112,0.4064],[114,0.5776],[116,0.7632],[118,1.0071],[120,1.2318],[122,1.4526],[124,1.5515],[126,1.7345],[128,1.8855],[130,1.9454],[132,1.9963],[134,1.966],[136,1.9428],[138,1.95],[140,1.9613],[142,1.9623],[144,2.0094],[146,2.0978],[148,2.1614],[152,2.2367],[154,2.2796],[156,2.3144],[158,2.3444],[160,2.3935],[162,2.4505],[164,2.4964],[166,2.5254],[168,2.5331],[170,2.5537],[172,2.5964],[174,2.6286],[176,2.6474],[182,2.6639],[184,2.6588],[186,2.6598],[188,2.6518],[194,2.6679],[198,2.6873],[200,2.6897],[202,2.6842],[204,2.6845],[206,2.6905],[212,2.6897],[214,2.681],[216,2.6648],[224,2.6463],[228,2.6563],[230,2.6698],[232,2.6627],[234,2.668],[238,2.6893],[242,2.6941],[248,2.6836],[250,2.6755],[252,2.6596],[256,2.6448],[260,2.6441],[264,2.6587],[266,2.6618],[268,2.6737],[272,2.7352],[274,2.7503],[276,2.7567],[278,2.7523],[282,2.7303],[284,2.7273],[288,2.7437],[292,2.7879],[294,2.8012],[298,2.8084],[300,2.8047],[302,2.7822],[306,2.7202],[308,2.6804],[312,2.6428],[314,2.6376],[316,2.6412],[320,2.6642],[328,2.687],[330,2.6841],[332,2.6417],[340,2.7136],[347,3.138],[354,4.1159],[356,4.5898],[360,4.9308],[362,4.4463],[364,1.7382],[366,0.5787],[368,0]]
       }
       calf_l_fk_ctrl.Rotation.Y {
        type=MovieSceneScriptingFloatChannel
        default=0
        thinned_from=132
        keys [[0,-0.0934],[92,-0.0934],[94,-0.0529],[96,-0.0105],[98,0.0067],[100,0],[102,0.022],[104,0.0685],[106,0.0474],[108,0.0965],[110,0.0442],[112,0.0952],[114,0.1159],[116,0.1356],[118,0.1301],[120,0.1164],[122,0.0713],[124,0.0645],[126,-0.053],[128,-0.1453],[130,-0.1608],[132,-0.1723],[134,-0.1257],[136,-0.0842],[138,-0.0626],[140,-0.0439],[142,-0.0121],[144,-0.0092],[146,-0.0323],[148,-0.0367],[150,-0.0229],[152,-0.0101],[154,-0.001],[156,0.0116],[158,0.027],[160,0.033],[162,0.036],[164,0.0449],[166,0.0598],[168,0.0834],[170,0.1009],[172,0.1083],[174,0.1196],[176,0.1352],[178,0.1545],[180,0.1741],[182,0.1925],[184,0.2136],[186,0.2316],[188,0.2524],[190,0.2673],[192,0.2797],[194,0.2896],[196,0.2979],[198,0.3045],[200,0.3127],[202,0.3241],[204,0.3329],[206,0.3399],[212,0.3703],[214,0.3845],[216,0.4018],[220,0.425],[224,0.4515],[228,0.469],[230,0.4757],[232,0.4915],[234,0.5017],[238,0.5166],[240,0.526],[242,0.5375],[244,0.5504],[248,0.5787],[250,0.5946],[252,0.6131],[256,0.6431],[260,0.6664],[264,0.6819],[266,0.6911],[268,0.696],[272,0.6918],[274,0.6963],[276,0.7044],[278,0.7163],[280,0.7309],[282,0.7437],[284,0.7542],[286,0.7597],[288,0.7633],[292,0.7612],[294,0.7632],[298,0.7734],[300,0.7806],[302,0.7948],[304,0.8117],[306,0.8313],[308,0.856],[310,0.8709],[312,0.8836],[314,0.8898],[316,0.8921],[320,0.8879],[324,0.8889],[326,0.8862],[328,0.8808],[330,0.8788],[332,0.8984],[340,0.8984],[347,0.7229],[354,0.2151],[356,-0.043],[358,-0.1329],[360,-0.1667],[362,0.1704],[364,0.1895],[366,0.0836],[368,0]]
       }
       calf_l_fk_ctrl.Rotation.Z {
        type=MovieSceneScriptingFloatChannel
        default=0
        thinned_from=132
        keys [[0,-1.4595],[92,-1.4595],[94,-0.907],[96,-0.287],[98,0.0053],[100,0],[102,1.1001],[104,3.8431],[106,6.7362],[108,11.0504],[110,14.1685],[112,18.512],[114,21.7893],[116,24.2587],[118,25.6054],[120,26.7217],[122,27.1345],[124,28.246],[126,27.6725],[128,27.5027],[130,28.2069],[132,28.8045],[134,30.0938],[136,30.9972],[138,31.1957],[140,31.0315],[142,30.7718],[144,29.8573],[148,26.7466],[150,25.3575],[152,23.8169],[154,22.0369],[158,18.3084],[160,16.1476],[162,13.8559],[164,11.6665],[166,9.597],[168,7.7664],[170,5.8027],[172,3.6172],[174,1.5502],[176,-0.328],[178,-2.0228],[180,-3.5807],[182,-5.0186],[184,-6.2133],[186,-7.3267],[188,-8.1786],[190,-8.9683],[192,-9.5966],[194,-10.0464],[196,-10.2803],[198,-10.3974],[200,-10.4485],[202,-10.384],[204,-10.3944],[206,-10.4527],[212,-10.4555],[214,-10.3612],[216,-10.1933],[220,-10.1455],[224,-10.035],[228,-10.1142],[230,-10.1882],[232,-10.0672],[234,-10.0793],[238,-10.2835],[240,-10.3393],[242,-10.347],[244,-10.3135],[248,-10.1558],[250,-10.0285],[252,-9.8443],[254,-9.7441],[256,-9.6124],[260,-9.5844],[262,-9.6499],[264,-9.75],[266,-9.7917],[268,-9.9544],[272,-10.6881],[274,-10.863],[276,-10.916],[278,-10.8426],[280,-10.6904],[282,-10.5873],[284,-10.5527],[286,-10.6389],[288,-10.7696],[292,-11.3055],[294,-11.4756],[298,-11.5719],[300,-11.529],[302,-11.2469],[304,-10.871],[306,-10.3919],[308,-9.7631],[310,-9.4243],[312,-9.1417],[314,-9.0263],[316,-8.9907],[318,-9.0664],[322,-9.085],[324,-9.0571],[326,-9.1201],[328,-9.2883],[330,-9.3576],[332,-8.9776],[340,15.359],[347,54.8758],[354,70.684],[356,72.5469],[360,73.5649],[362,73.3251],[364,31.7669],[366,11.0952],[368,0]]
       }
       calf_l_fk_ctrl.Scale.X {type=MovieSceneScriptingFloatChannel, default=1, const [1,132]}
       calf_l_fk_ctrl.Scale.Y {type=MovieSceneScriptingFloatChannel, default=1, const [1,132]}
       calf_l_fk_ctrl.Scale.Z {type=MovieSceneScriptingFloatChannel, default=1, const [1,132]}
       foot_l_fk_ctrl.Location.X {
        type=MovieSceneScriptingFloatChannel
        default=0
        thinned_from=140
        keys [[0,0],[108,-0.0001],[116,0],[118,0.0001],[332,0.0001],[334,0],[336,0.0004],[340,0.0006],[342,0.0005],[344,0.0001],[368,0]]
       }
       foot_l_fk_ctrl.Location.Y {
        type=MovieSceneScriptingFloatChannel
        default=0
        thinned_from=140
        keys [[0,0],[344,0],[346,0.0014],[348,0.005],[350,0.0103],[352,0.0166],[354,0.0232],[356,0.0295],[358,0.0348],[360,0.0384],[362,0.0398],[364,0.0173],[366,0.0061],[368,0]]
       }
       foot_l_fk_ctrl.Location.Z {
        type=MovieSceneScriptingFloatChannel
        default=0
        thinned_from=140
        keys [[0,0],[338,-0.0001],[340,-0.0004],[344,-0.0001],[346,-0.0305],[348,-0.1122],[350,-0.2304],[352,-0.3705],[354,-0.518],[356,-0.6581],[358,-0.7763],[360,-0.858],[362,-0.8885],[364,-0.3868],[366,-0.1354],[368,0]]
       }
       foot_l_fk_ctrl.Rotation.X {
        type=MovieSceneScriptingFloatChannel
        default=0
        thinned_from=140
        keys [[0,2.7604],[92,2.7604],[94,2.0938],[96,1.2949],[98,0.6309],[100,0],[102,-0.2605],[104,-0.5782],[106,-0.6306],[108,-0.1242],[110,1.1886],[112,2.029],[114,2.9201],[116,3.1514],[118,2.6165],[120,2.0038],[122,2.0614],[124,2.1047],[126,2.5519],[128,3.0505],[130,3.043],[132,2.6116],[134,1.7046],[136,0.842],[138,-0.2236],[140,-1.8416],[142,-2.9424],[144,-4.6067],[146,-5.5442],[148,-5.9875],[150,-6.5157],[152,-7.008],[156,-7.7076],[158,-8.0178],[162,-8.6747],[166,-9.7151],[168,-9.8584],[170,-9.976],[172,-10.2434],[174,-10.2459],[176,-10.1081],[178,-9.4715],[180,-8.9591],[182,-8.7252],[184,-8.7233],[188,-8.9795],[190,-9.1527],[192,-9.6236],[194,-10.383],[196,-10.9794],[198,-11.2094],[200,-11.6134],[202,-11.7553],[204,-11.8148],[206,-11.7746],[208,-11.8211],[210,-11.8243],[212,-11.7781],[214,-11.5687],[216,-11.3321],[220,-11.533],[222,-11.5861],[224,-11.5372],[226,-11.6361],[228,-11.8059],[230,-11.7812],[246,-10.8047],[284,-8.8134],[294,-8.2096],[302,-7.6622],[308,-7.2041],[314,-6.6979],[318,-6.3306],[322,-5.9371],[324,-5.6784],[326,-5.3183],[328,-4.8696],[330,-4.3452],[332,-3.6165],[334,-2.7387],[336,-1.5604],[338,-1.2288],[340,-0.8532],[342,-2.2889],[344,-4.597],[346,-5.5315],[348,-6.6441],[350,-7.8669],[352,-9.1324],[354,-10.3728],[356,-11.5203],[358,-12.5073],[360,-13.1974],[362,-13.4455],[364,-5.8535],[366,-2.0487],[368,0]]
       }
       foot_l_fk_ctrl.Rotation.Y {
        type=MovieSceneScriptingFloatChannel
        default=0
        thinned_from=140
        keys [[0,0.6705],[92,0.6705],[94,0.6347],[96,0.4629],[98,0.3169],[100,0],[102,-0.1866],[104,-0.3989],[106,-0.4823],[108,-0.2639],[110,0.1059],[112,0.6016],[114,1.0026],[116,1.0655],[118,0.8773],[120,0.8682],[122,1.1943],[124,1.3415],[126,1.5312],[128,1.7724],[130,1.876],[132,1.884],[134,1.8435],[136,1.9302],[138,1.9214],[140,1.7242],[142,1.5098],[144,0.6094],[146,-0.08],[148,-0.3653],[150,-0.4992],[152,-0.7869],[154,-0.9061],[156,-0.9033],[158,-0.8898],[160,-1.006],[162,-1.1443],[164,-1.333],[166,-1.301],[168,-1.1748],[170,-0.975],[172,-0.8167],[174,-0.5755],[176,-0.3149],[178,0.1092],[180,0.5265],[182,0.8101],[184,0.9668],[186,1.0372],[188,1.115],[190,1.141],[192,1.0333],[194,0.7868],[196,0.5657],[198,0.4653],[200,0.407],[202,0.3753],[204,0.4135],[206,0.4208],[208,0.4493],[210,0.463],[212,0.5968],[214,0.7214],[216,0.8316],[218,0.9005],[220,0.9062],[222,0.8402],[224,0.7554],[226,0.5173],[228,0.3643],[230,0.0475],[232,-0.0189],[234,-0.0439],[236,-0.0704],[238,-0.0984],[240,-0.1276],[242,-0.1579],[244,-0.1892],[246,-0.2213],[250,-0.2871],[260,-0.4543],[264,-0.5189],[268,-0.5804],[270,-0.6095],[272,-0.6374],[274,-0.6638],[276,-0.6887],[278,-0.7118],[280,-0.7331],[282,-0.7522],[284,-0.7692],[286,-0.7838],[288,-0.7958],[290,-0.8052],[292,-0.8117],[294,-0.8152],[296,-0.8155],[298,-0.8126],[300,-0.8061],[302,-0.796],[304,-0.7822],[306,-0.7643],[308,-0.7424],[310,-0.7162],[312,-0.6856],[314,-0.6504],[316,-0.6104],[318,-0.5656],[320,-0.5157],[322,-0.4606],[324,-0.393],[326,-0.307],[328,-0.2044],[330,-0.0867],[332,0.206],[334,0.3713],[336,-0.4271],[338,0.8206],[340,3.532],[342,4.7741],[344,5.9208],[346,6.636],[350,8.17],[352,8.9158],[354,9.5989],[356,10.1827],[358,10.6308],[360,10.9183],[362,11.0218],[364,4.7983],[366,1.6794],[368,0]]
       }
       foot_l_fk_ctrl.Rotation.Z {
        type=MovieSceneScriptingFloatChannel
        default=29.9903
        thinned_from=140
        keys [[0,41.2849],[92,41.2849],[96,40.6065],[100,39.8411],[102,39.6244],[104,39.1626],[106,38.4988],[108,38.1262],[110,38.3722],[116,38.3593],[120,38.5702],[122,38.7927],[124,38.8378],[126,39.1104],[128,39.2469],[130,39.106],[132,38.6368],[136,37.3684],[140,36.3792],[144,35.6529],[158,35.1665],[168,35.182],[176,35.5991],[180,35.9036],[182,35.9692],[190,35.9999],[196,35.8872],[206,36.08],[208,36.0304],[210,36.0813],[212,36.296],[220,36.4281],[230,36.3135],[310,38.5344],[322,38.574],[324,38.449],[326,38.1019],[328,37.5604],[330,36.8524],[332,35.7858],[334,34.1031],[336,30.8145],[338,26.3371],[340,22.2231],[342,18.8429],[344,16.5876],[348,13.1902],[350,11.5882],[352,10.1381],[354,8.9055],[356,7.9559],[358,7.3549],[360,7.0387],[362,6.9249],[364,19.9488],[366,26.4758],[368,29.9903]]
       }
       foot_l_fk_ctrl.Scale.X {type=MovieSceneScriptingFloatChannel, default=1, const [1,140]}
       foot_l_fk_ctrl.Scale.Y {type=MovieSceneScriptingFloatChannel, default=1, const [1,140]}
       foot_l_fk_ctrl.Scale.Z {type=MovieSceneScriptingFloatChannel, default=1, const [1,140]}
       ball_l_fk_ctrl.Location.X {type=MovieSceneScriptingFloatChannel, default=0.0007, thinned_from=140, keys [[0,0],[362,0],[368,0.0007]]}
       ball_l_fk_ctrl.Location.Y {type=MovieSceneScriptingFloatChannel, default=-0.0013, thinned_from=140, keys [[0,0],[362,0],[364,-0.0006],[366,-0.0007],[368,-0.0013]]}
       ball_l_fk_ctrl.Location.Z {type=MovieSceneScriptingFloatChannel, default=-0.0006, thinned_from=140, keys [[0,0],[362,0],[364,-0.0002],[366,-0.0003],[368,-0.0006]]}
       ball_l_fk_ctrl.Rotation.X {
        type=MovieSceneScriptingFloatChannel
        default=-0.0482
        thinned_from=140
        keys [[0,-0.163],[92,-0.163],[94,-0.1593],[96,-0.1526],[98,-0.1485],[100,-0.1447],[102,-0.1425],[104,-0.1418],[106,-0.1509],[108,-0.1574],[110,-0.1546],[112,-0.1574],[114,-0.1621],[116,-0.1659],[118,-0.1678],[120,-0.169],[122,-0.1666],[124,-0.1657],[126,-0.1638],[128,-0.1629],[130,-0.1636],[132,-0.1631],[134,-0.1618],[136,-0.1612],[138,-0.1601],[140,-0.1618],[142,-0.1689],[144,-0.1831],[146,-0.1902],[148,-0.1924],[150,-0.192],[152,-0.1922],[154,-0.1936],[156,-0.1941],[158,-0.1941],[160,-0.1954],[162,-0.1961],[164,-0.1959],[168,-0.1944],[170,-0.1913],[172,-0.1895],[174,-0.1866],[178,-0.1824],[182,-0.177],[188,-0.1754],[190,-0.1759],[192,-0.1798],[196,-0.184],[198,-0.1852],[202,-0.1843],[204,-0.1828],[206,-0.1827],[208,-0.1819],[210,-0.1816],[212,-0.1796],[214,-0.179],[218,-0.1793],[220,-0.1804],[224,-0.1849],[226,-0.1884],[228,-0.1894],[230,-0.1909],[232,-0.1895],[236,-0.1886],[238,-0.1877],[240,-0.1881],[242,-0.1875],[250,-0.1884],[252,-0.1881],[254,-0.1893],[256,-0.1912],[258,-0.1904],[260,-0.1918],[264,-0.1915],[266,-0.191],[268,-0.1876],[270,-0.1852],[272,-0.187],[274,-0.1882],[278,-0.1878],[280,-0.1881],[284,-0.1873],[286,-0.1851],[288,-0.1821],[290,-0.18],[292,-0.1802],[294,-0.1794],[296,-0.1796],[300,-0.1834],[302,-0.1875],[304,-0.1898],[306,-0.1948],[308,-0.1982],[312,-0.1972],[314,-0.196],[316,-0.1937],[318,-0.1904],[322,-0.1884],[324,-0.1864],[328,-0.1833],[332,-0.1864],[334,-0.1906],[336,-0.1704],[338,-0.1617],[340,-0.1571],[342,-0.151],[344,-0.1463],[346,-0.1403],[348,-0.1295],[350,-0.1086],[352,-0.0806],[354,-0.0448],[356,-0.0168],[358,0.0008],[360,0.0106],[362,0.0236],[364,-0.0208],[366,-0.0917],[368,-0.0482]]
       }
       ball_l_fk_ctrl.Rotation.Y {
        type=MovieSceneScriptingFloatChannel
        default=0.0902
        thinned_from=140
        keys [[0,0.3269],[92,0.3269],[94,0.3188],[96,0.3043],[98,0.2954],[100,0.2874],[102,0.2825],[104,0.2812],[106,0.3007],[108,0.3148],[110,0.3086],[112,0.3148],[114,0.3248],[116,0.3332],[118,0.3374],[120,0.3399],[122,0.3348],[124,0.3327],[126,0.3285],[128,0.3267],[130,0.3281],[132,0.327],[134,0.3243],[136,0.323],[138,0.3206],[140,0.3243],[142,0.3396],[144,0.3711],[146,0.3869],[148,0.3921],[150,0.3911],[152,0.3915],[154,0.3946],[156,0.3957],[158,0.3958],[160,0.3988],[162,0.4003],[164,0.3999],[166,0.3977],[168,0.3964],[170,0.3895],[172,0.3854],[174,0.3791],[178,0.3696],[182,0.3577],[188,0.3541],[190,0.3552],[192,0.3639],[196,0.3733],[198,0.3759],[202,0.3739],[204,0.3706],[206,0.3703],[208,0.3685],[210,0.3677],[212,0.3634],[214,0.3621],[218,0.3628],[220,0.3651],[224,0.3753],[226,0.383],[228,0.3854],[230,0.3887],[232,0.3855],[236,0.3835],[238,0.3815],[240,0.3823],[242,0.381],[250,0.383],[252,0.3824],[254,0.385],[256,0.3893],[258,0.3876],[260,0.3906],[264,0.39],[266,0.3889],[268,0.3812],[270,0.376],[272,0.3798],[274,0.3825],[278,0.3817],[280,0.3823],[284,0.3807],[286,0.3756],[288,0.369],[290,0.3644],[292,0.3646],[294,0.3629],[296,0.3634],[300,0.3719],[302,0.3809],[304,0.3861],[306,0.3974],[308,0.4051],[312,0.4028],[314,0.4],[316,0.3949],[318,0.3876],[322,0.3831],[324,0.3785],[328,0.3717],[332,0.3786],[334,0.388],[336,0.3431],[338,0.324],[340,0.314],[342,0.3009],[344,0.2907],[346,0.278],[348,0.255],[350,0.2115],[352,0.1543],[354,0.0836],[356,0.0296],[358,-0.0035],[360,-0.0218],[362,-0.0459],[364,0.038],[366,0.1781],[368,0.0902]]
       }
       ball_l_fk_ctrl.Rotation.Z {
        type=MovieSceneScriptingFloatChannel
        default=1.1101
        thinned_from=140
        keys [[0,3.5829],[92,3.5829],[94,3.4993],[96,3.3489],[98,3.257],[100,3.1734],[102,3.1233],[104,3.1093],[106,3.3114],[108,3.4574],[110,3.3935],[112,3.4571],[114,3.5613],[116,3.6479],[118,3.6914],[120,3.7171],[122,3.6641],[124,3.6431],[126,3.5995],[128,3.5802],[130,3.5949],[132,3.584],[134,3.5558],[136,3.5424],[138,3.5175],[140,3.5558],[142,3.7142],[144,4.0388],[146,4.2011],[148,4.2543],[150,4.244],[152,4.2487],[154,4.2808],[156,4.2918],[158,4.2924],[160,4.3233],[162,4.3391],[164,4.3343],[168,4.2991],[170,4.2282],[172,4.1859],[174,4.1204],[178,4.0235],[182,3.9003],[188,3.8639],[190,3.8743],[192,3.9642],[196,4.0608],[198,4.0879],[202,4.067],[204,4.0334],[206,4.0308],[208,4.012],[210,4.0038],[212,3.9591],[214,3.9457],[218,3.9527],[220,3.9772],[224,4.0814],[226,4.1612],[228,4.1853],[230,4.2196],[232,4.1871],[236,4.166],[238,4.1454],[240,4.1534],[242,4.1402],[250,4.1612],[252,4.1547],[254,4.1818],[256,4.2262],[258,4.2085],[260,4.2397],[264,4.2332],[266,4.2213],[268,4.1428],[270,4.0887],[272,4.1279],[274,4.1558],[278,4.1479],[280,4.1535],[284,4.1369],[286,4.0853],[288,4.0171],[290,3.9691],[292,3.972],[294,3.9543],[296,3.9597],[300,4.0469],[302,4.1392],[304,4.1935],[306,4.3091],[308,4.3879],[312,4.3643],[314,4.3362],[316,4.2836],[318,4.2085],[322,4.1617],[324,4.1142],[328,4.0448],[332,4.1159],[334,4.2128],[336,3.7499],[338,3.5526],[340,3.4494],[342,3.3136],[344,3.2083],[346,3.0766],[348,2.8377],[350,2.3842],[352,1.7857],[354,1.0412],[356,0.4701],[358,0.1182],[360,-0.0769],[362,-0.3345],[364,0.5567],[366,2.0321],[368,1.1101]]
       }
       ball_l_fk_ctrl.Scale.X {type=MovieSceneScriptingFloatChannel, default=1, const [1,140]}
       ball_l_fk_ctrl.Scale.Y {type=MovieSceneScriptingFloatChannel, default=1, const [1,140]}
       ball_l_fk_ctrl.Scale.Z {type=MovieSceneScriptingFloatChannel, default=1, const [1,140]}
       foot_l_ik_ctrl.Location.X {
        type=MovieSceneScriptingFloatChannel
        default=-9.063
        thinned_from=140
        keys [[0,34.9635],[92,34.9635],[94,35.5737],[96,36.0646],[98,36.2797],[100,36.1973],[104,36.3808],[106,35.2048],[108,34.4506],[110,32.7674],[112,32.3794],[114,31.7356],[116,31.5282],[118,30.9868],[120,30.3179],[122,29.5068],[124,29.3564],[126,29.0029],[130,28.8794],[132,28.6407],[134,27.9593],[136,26.7579],[140,24.7037],[142,23.877],[144,23.375],[146,23.1513],[148,22.8662],[150,22.5196],[152,22.1047],[154,21.4761],[158,19.9274],[160,19.2811],[162,18.6827],[164,18.19],[168,17.3129],[170,16.8147],[172,16.2757],[176,14.98],[180,13.4248],[182,12.7467],[184,12.1915],[186,11.7299],[188,11.3232],[190,10.9916],[192,10.8231],[196,10.874],[198,10.8328],[202,10.8999],[206,10.9026],[208,10.8842],[210,10.9078],[216,10.7939],[222,11.0279],[224,11.074],[226,11.2034],[230,11.5153],[234,11.6459],[238,11.8295],[244,11.9604],[246,12.0749],[254,12.2984],[256,12.3919],[258,12.414],[264,12.3502],[268,12.3564],[272,12.2771],[276,12.351],[280,12.4794],[282,12.4899],[284,12.4555],[286,12.3849],[292,12.266],[294,12.244],[298,12.2483],[300,12.2875],[304,12.4531],[306,12.6325],[308,12.8774],[312,13.1725],[314,13.2551],[318,13.5384],[320,13.7438],[322,13.8012],[324,13.9601],[326,14.0463],[328,13.9574],[330,13.7607],[332,13.5241],[334,13.2505],[336,12.7255],[338,12.2718],[340,11.4509],[342,11.5371],[344,11.3327],[346,10.0074],[348,8.4009],[350,7.701],[352,7.5575],[354,7.9021],[356,8.8527],[358,9.7589],[360,11.9726],[362,14.7199],[364,20.6124],[366,15.4349],[368,-9.063]]
       }
       foot_l_ik_ctrl.Location.Y {
        type=MovieSceneScriptingFloatChannel
        default=73.0443
        thinned_from=140
        keys [[0,-1060.7451],[94,-1060.4431],[96,-985.7559],[98,-889.9875],[100,-781.4783],[102,-668.6526],[104,-559.3441],[106,-461.9978],[108,-385.337],[110,-337.0971],[112,-307.5365],[114,-280.7251],[116,-256.5349],[118,-234.9043],[120,-215.7457],[122,-198.9281],[124,-184.3115],[126,-171.4836],[128,-160.0383],[130,-150.0103],[132,-141.024],[134,-133.0957],[136,-126.1276],[150,-79.581],[152,-73.1308],[154,-66.9908],[156,-61.178],[158,-55.6447],[160,-50.4213],[162,-45.5928],[164,-41.1719],[168,-32.8621],[170,-28.8625],[172,-25.0339],[174,-21.4301],[176,-18.0022],[178,-14.8006],[180,-11.8314],[182,-9.0731],[184,-6.5369],[186,-4.2592],[188,-2.2619],[190,-0.5695],[192,0.8176],[194,1.8845],[196,2.5673],[198,2.9344],[200,3.0921],[202,3.1512],[204,3.159],[206,3.134],[210,3.0671],[212,3.0388],[216,2.9972],[218,3.0121],[220,3.0183],[224,2.9976],[226,3.0126],[228,3.0539],[230,3.1253],[236,3.2658],[238,3.2963],[244,3.2942],[248,3.344],[250,3.361],[258,3.3627],[260,3.3581],[262,3.3466],[264,3.3172],[266,3.3016],[270,3.2582],[274,3.2064],[278,3.1906],[280,3.2164],[282,3.2135],[284,3.1958],[286,3.1846],[290,3.1717],[292,3.1422],[294,3.1],[296,3.0797],[298,3.0205],[300,2.9856],[302,2.9569],[304,2.9478],[306,2.975],[308,3.0338],[314,3.1567],[320,3.3262],[322,3.335],[324,3.3537],[326,3.3601],[328,3.338],[330,3.284],[332,3.441],[334,3.926],[336,5.5295],[338,7.4469],[340,8.5114],[342,8.8005],[344,8.4038],[346,9.958],[348,11.2859],[350,11.7376],[352,11.1468],[354,10.3359],[356,9.3975],[358,8.0418],[360,3.5093],[362,-4.9954],[364,-14.2449],[366,-7.3598],[368,73.0443]]
       }
       foot_l_ik_ctrl.Location.Z {
        type=MovieSceneScriptingFloatChannel
        default=119.9886
        thinned_from=140
        keys [[0,89.8383],[96,89.9045],[100,90.41],[104,89.8155],[106,90.089],[108,88.9008],[110,88.2587],[114,83.774],[120,75.4237],[122,73.0202],[124,70.4369],[126,68.4412],[128,66.0656],[132,60.374],[134,57.1483],[136,53.637],[142,42.4111],[144,38.7897],[146,35.2546],[148,31.8272],[150,28.5883],[152,25.569],[154,22.7578],[156,20.1817],[158,17.8496],[160,15.7628],[162,13.9276],[164,12.3333],[166,10.8863],[168,9.5101],[170,8.2193],[172,7.0327],[174,5.9472],[176,4.9726],[178,4.0881],[180,3.3018],[182,2.6148],[184,2.0131],[186,1.5212],[188,1.0913],[190,0.7434],[192,0.4624],[194,0.2471],[196,0.0907],[198,-0.0336],[200,-0.1269],[202,-0.1927],[204,-0.2184],[206,-0.2055],[208,-0.1552],[210,-0.0675],[212,0.0433],[214,0.1674],[216,0.2995],[218,0.4803],[220,0.6814],[222,0.895],[224,1.1269],[226,1.3844],[228,1.6603],[230,1.9527],[232,2.2398],[234,2.5541],[238,3.2136],[242,3.8414],[246,4.4482],[248,4.7406],[250,5.0161],[254,5.514],[258,5.9744],[260,6.1843],[262,6.3781],[264,6.5511],[268,6.8274],[270,6.947],[272,7.0335],[274,7.0688],[276,7.065],[278,7.0363],[288,6.7791],[292,6.6903],[294,6.6202],[298,6.4424],[300,6.334],[308,5.7732],[312,5.5628],[314,5.4839],[316,5.4305],[318,5.3938],[320,5.3679],[322,5.3192],[328,5.2427],[330,5.1955],[332,5.0866],[334,4.8166],[336,3.8479],[338,2.9003],[340,2.1774],[342,2.0762],[344,2.3944],[346,2.1774],[348,2.5614],[350,3.3471],[352,4.5205],[354,5.935],[356,7.2482],[358,8.1017],[360,7.72],[362,7.0521],[364,7.6083],[366,37.1462],[368,119.9886]]
       }
       foot_l_ik_ctrl.Rotation.X {
        type=MovieSceneScriptingFloatChannel
        default=84.8417
        thinned_from=140
        keys [[0,133.5061],[94,133.6332],[102,133.8932],[108,136.0064],[110,136.941],[112,137.3749],[114,136.9304],[116,135.6375],[124,127.7059],[128,123.1484],[130,120.6746],[134,115.0628],[136,112.0827],[138,108.7793],[142,101.6571],[148,89.9298],[154,78.8579],[156,75.3344],[160,68.6575],[162,65.5476],[164,62.6875],[172,52.2149],[174,49.7232],[176,47.3563],[178,45.2147],[182,41.2342],[184,39.3983],[186,37.7765],[188,36.4018],[190,35.2322],[192,34.2215],[194,33.4293],[196,33.0028],[198,32.8534],[200,32.806],[206,32.9728],[208,32.9497],[210,33.0328],[212,33.2807],[216,33.5677],[224,33.6436],[228,33.5299],[230,33.5256],[234,33.7689],[244,34.1216],[256,34.924],[262,35.1328],[266,35.1899],[268,35.1379],[272,34.8609],[276,34.8257],[284,35.1058],[288,35.0807],[294,34.8837],[300,34.9895],[304,35.3838],[306,35.6581],[308,36.0152],[312,36.4271],[316,36.6138],[322,36.6015],[324,36.4592],[326,36.0402],[328,35.3835],[330,34.6122],[332,33.4839],[334,32.106],[336,31.058],[338,29.0412],[340,28.5974],[342,29.6056],[344,31.2823],[346,31.1101],[348,31.2976],[350,32.2526],[352,34.0083],[356,37.6817],[358,39.2435],[360,43.3796],[362,50.6974],[364,59.684],[366,69.8174],[368,84.8417]]
       }
       foot_l_ik_ctrl.Rotation.Y {
        type=MovieSceneScriptingFloatChannel
        default=-0.7187
        thinned_from=140
        keys [[0,-5.0599],[92,-5.0599],[94,-4.8395],[96,-4.2743],[98,-3.5293],[100,-2.6527],[102,-2.0462],[104,-1.0411],[106,0.4781],[108,1.3374],[110,1.9362],[112,2.1359],[114,2.2677],[116,2.7569],[118,3.4887],[120,3.8024],[122,3.4403],[124,3.4118],[126,3.5606],[128,3.9218],[130,4.0954],[132,4.3544],[134,4.1715],[136,3.4912],[138,2.7992],[140,1.9165],[142,0.8863],[144,-0.1338],[146,-0.8969],[148,-1.6331],[150,-2.3946],[152,-3.1772],[154,-3.9999],[156,-4.8654],[158,-5.6342],[160,-6.3283],[162,-6.9676],[164,-7.6458],[166,-8.1542],[168,-8.3537],[170,-8.4734],[172,-8.6843],[174,-8.7524],[176,-8.6863],[178,-8.2966],[180,-7.9659],[182,-7.7859],[184,-7.7363],[186,-7.7722],[188,-7.7857],[190,-7.859],[192,-8.1241],[194,-8.5645],[196,-8.9554],[198,-9.1667],[200,-9.4375],[202,-9.5535],[204,-9.5889],[206,-9.605],[208,-9.6464],[210,-9.6454],[212,-9.5798],[216,-9.2582],[218,-9.2276],[220,-9.2435],[222,-9.298],[224,-9.3316],[226,-9.4981],[228,-9.5995],[230,-9.7399],[234,-9.5759],[238,-9.3337],[244,-9.0988],[246,-8.9561],[248,-8.8679],[252,-8.7354],[254,-8.6556],[256,-8.547],[258,-8.4932],[264,-8.4493],[266,-8.4069],[268,-8.3849],[272,-8.4462],[274,-8.4136],[276,-8.3538],[278,-8.2725],[280,-8.1724],[282,-8.1191],[284,-8.0939],[286,-8.0914],[290,-8.0299],[292,-8.0212],[294,-7.9864],[298,-7.8341],[300,-7.7169],[302,-7.5483],[304,-7.3471],[306,-7.0494],[308,-6.6592],[312,-6.0401],[314,-5.7877],[318,-5.1609],[320,-4.7769],[322,-4.5379],[326,-3.7454],[328,-3.4444],[330,-3.1865],[332,-2.6284],[334,-1.9899],[336,-1.5619],[338,0.0453],[340,2.5462],[342,3.2208],[344,2.726],[346,2.0664],[348,1.1829],[350,0.8664],[352,0.716],[354,0.5961],[356,0.8424],[358,0.6878],[360,0.203],[362,-1.3871],[364,-0.0362],[366,-2.9028],[368,-0.7187]]
       }
       foot_l_ik_ctrl.Rotation.Z {
        type=MovieSceneScriptingFloatChannel
        default=-0.259
        thinned_from=140
        keys [[0,12.5309],[92,12.5309],[94,12.5749],[96,12.4012],[98,12.1328],[100,11.5753],[102,11.4697],[104,11.4967],[106,10.9691],[108,11.2939],[110,11.6759],[112,12.7521],[114,13.6006],[116,14.0341],[118,13.4218],[120,12.7382],[122,12.7183],[124,12.9766],[126,13.2981],[128,13.8436],[130,14.0269],[132,13.5772],[134,12.5562],[136,11.3934],[138,10.0854],[140,8.3242],[142,7.3766],[144,5.6539],[146,4.9179],[148,5.0114],[150,5.2677],[152,5.4942],[154,5.9434],[156,6.4806],[158,7.0744],[162,8.1196],[164,8.4753],[166,9.0302],[168,9.8292],[170,10.6541],[172,11.3563],[174,12.1768],[176,13.0626],[178,14.1062],[180,15.0099],[182,15.6872],[184,16.1674],[186,16.5342],[188,16.8364],[190,17.0554],[192,17.1421],[194,17.0697],[196,16.9356],[202,16.649],[206,16.551],[208,16.5553],[210,16.4859],[216,16.4797],[218,16.4377],[224,16.1106],[226,15.8703],[228,15.666],[230,15.3743],[232,15.2374],[234,15.1765],[236,14.9963],[238,14.7266],[240,14.5016],[242,14.3188],[244,14.069],[246,13.7189],[248,13.5202],[250,13.394],[254,12.9556],[256,12.6201],[258,12.4692],[262,12.392],[264,12.289],[266,12.1169],[268,11.8363],[272,11.7263],[276,11.3628],[280,10.9371],[282,10.7909],[284,10.7262],[286,10.7291],[288,10.6507],[290,10.5337],[292,10.463],[296,10.143],[298,9.9567],[300,9.7441],[304,9.2022],[306,8.8682],[308,8.488],[310,8.2193],[312,7.9831],[314,7.8729],[316,7.6578],[318,7.3744],[320,6.962],[322,6.7644],[324,6.3254],[326,5.9671],[328,5.8411],[330,5.9472],[332,5.8626],[334,5.6636],[336,5.1901],[338,5.5022],[340,6.8635],[342,7.267],[344,8.0799],[346,9.2178],[348,10.6704],[350,11.9818],[352,12.7274],[354,13.0485],[356,13.4338],[358,13.1207],[360,13.4848],[362,13.4579],[364,13.4843],[366,8.9913],[368,-0.259]]
       }
       foot_l_ik_ctrl.Scale.X {type=MovieSceneScriptingFloatChannel, default=1, const [1,140]}
       foot_l_ik_ctrl.Scale.Y {type=MovieSceneScriptingFloatChannel, default=1, const [1,140]}
       foot_l_ik_ctrl.Scale.Z {type=MovieSceneScriptingFloatChannel, default=1, const [1,140]}
       ball_l_ik_ctrl.Location.X {type=MovieSceneScriptingFloatChannel, default=0.0005, thinned_from=140, keys [[0,0],[366,0],[368,0.0005]]}
       ball_l_ik_ctrl.Location.Y {type=MovieSceneScriptingFloatChannel, default=-0.0013, thinned_from=140, keys [[0,0],[364,0],[366,-0.0002],[368,-0.0013]]}
       ball_l_ik_ctrl.Location.Z {type=MovieSceneScriptingFloatChannel, default=0.0008, thinned_from=140, keys [[0,0],[364,0],[368,0.0008]]}
       ball_l_ik_ctrl.Rotation.X {
        type=MovieSceneScriptingFloatChannel
        default=1.1137
        thinned_from=140
        keys [[0,3.5976],[92,3.5976],[94,3.5136],[96,3.3625],[98,3.2703],[100,3.1863],[102,3.1359],[104,3.1218],[106,3.3249],[108,3.4716],[110,3.4073],[112,3.4712],[114,3.576],[116,3.6629],[118,3.7066],[120,3.7325],[122,3.6792],[124,3.6581],[126,3.6143],[128,3.5949],[130,3.6097],[132,3.5988],[134,3.5704],[136,3.557],[138,3.532],[140,3.5704],[142,3.7295],[144,4.0556],[146,4.2187],[148,4.2722],[150,4.2618],[152,4.2665],[154,4.2987],[156,4.3098],[158,4.3104],[160,4.3415],[162,4.3573],[164,4.3525],[168,4.3171],[170,4.2459],[172,4.2034],[174,4.1377],[178,4.0403],[182,3.9164],[188,3.8799],[190,3.8903],[192,3.9806],[196,4.0777],[198,4.105],[202,4.084],[204,4.0502],[206,4.0476],[208,4.0287],[210,4.0205],[212,3.9755],[214,3.9621],[218,3.9691],[220,3.9937],[224,4.0985],[226,4.1786],[228,4.2028],[230,4.2373],[232,4.2046],[236,4.1834],[238,4.1627],[240,4.1707],[242,4.1575],[250,4.1786],[252,4.1721],[254,4.1993],[256,4.2439],[258,4.2261],[260,4.2575],[264,4.2509],[266,4.239],[268,4.1601],[270,4.1057],[272,4.1451],[274,4.1732],[278,4.1652],[280,4.1709],[284,4.1542],[286,4.1023],[288,4.0339],[290,3.9856],[292,3.9885],[294,3.9707],[296,3.9761],[300,4.0638],[302,4.1565],[304,4.211],[306,4.3272],[308,4.4064],[312,4.3826],[314,4.3544],[316,4.3015],[318,4.2261],[322,4.1791],[324,4.1314],[328,4.0616],[332,4.1331],[334,4.2304],[336,3.7654],[338,3.5671],[340,3.4636],[342,3.3271],[344,3.2213],[346,3.089],[348,2.8491],[350,2.3935],[352,1.7923],[354,1.0445],[356,0.4709],[358,0.1175],[360,-0.0785],[362,-0.3371],[364,0.168],[366,2.6894],[368,1.1137]]
       }
       ball_l_ik_ctrl.Rotation.Y {
        type=MovieSceneScriptingFloatChannel
        default=-0.007
        thinned_from=140
        keys [[0,0.0056],[92,0.0056],[94,0.0053],[96,0.0047],[100,0.0039],[104,0.0036],[106,0.0045],[108,0.0051],[110,0.0049],[116,0.0059],[120,0.0062],[128,0.0056],[138,0.0054],[140,0.0055],[142,0.0061],[144,0.0074],[146,0.0079],[148,0.0081],[164,0.0084],[168,0.0083],[182,0.0068],[190,0.0068],[198,0.0075],[218,0.007],[230,0.008],[252,0.0078],[256,0.008],[266,0.008],[270,0.0075],[284,0.0077],[290,0.0071],[296,0.0071],[300,0.0074],[308,0.0086],[316,0.0082],[328,0.0074],[332,0.0076],[334,0.008],[336,0.0063],[338,0.0055],[346,0.0035],[348,0.0024],[350,0.0002],[352,-0.003],[354,-0.0074],[356,-0.0111],[358,-0.0136],[360,-0.015],[362,-0.0169],[364,-0.0131],[366,0.0017],[368,-0.007]]
       }
       ball_l_ik_ctrl.Rotation.Z {
        type=MovieSceneScriptingFloatChannel
        default=0.0498
        thinned_from=140
        keys [[0,0.1762],[92,0.1762],[94,0.1719],[96,0.1642],[98,0.1595],[100,0.1552],[102,0.1527],[104,0.1519],[106,0.1623],[108,0.1698],[110,0.1665],[112,0.1698],[114,0.1751],[116,0.1796],[118,0.1818],[120,0.1831],[122,0.1804],[124,0.1793],[126,0.1771],[128,0.1761],[130,0.1768],[132,0.1763],[134,0.1748],[136,0.1741],[138,0.1729],[140,0.1748],[142,0.183],[144,0.1996],[146,0.2079],[148,0.2106],[150,0.2101],[152,0.2104],[154,0.212],[156,0.2126],[158,0.2126],[160,0.2142],[162,0.215],[164,0.2148],[166,0.2136],[168,0.2129],[170,0.2093],[172,0.2071],[174,0.2038],[178,0.1988],[182,0.1925],[188,0.1906],[190,0.1912],[192,0.1958],[196,0.2007],[198,0.2021],[202,0.201],[204,0.1993],[206,0.1992],[208,0.1982],[210,0.1978],[212,0.1955],[214,0.1948],[218,0.1952],[220,0.1964],[224,0.2018],[226,0.2059],[228,0.2071],[230,0.2089],[232,0.2072],[236,0.2061],[238,0.2051],[240,0.2055],[242,0.2048],[250,0.2059],[252,0.2055],[254,0.2069],[256,0.2092],[258,0.2083],[260,0.2099],[264,0.2096],[266,0.209],[268,0.2049],[270,0.2021],[272,0.2042],[274,0.2056],[278,0.2052],[280,0.2055],[284,0.2046],[286,0.202],[288,0.1985],[290,0.196],[292,0.1962],[294,0.1953],[296,0.1955],[300,0.2],[302,0.2047],[304,0.2075],[306,0.2135],[308,0.2175],[312,0.2163],[314,0.2148],[316,0.2121],[318,0.2083],[322,0.2059],[324,0.2035],[328,0.1999],[332,0.2035],[334,0.2085],[336,0.1847],[338,0.1747],[340,0.1694],[342,0.1624],[344,0.157],[346,0.1503],[348,0.1381],[350,0.1148],[352,0.0843],[354,0.0463],[356,0.0171],[358,-0.0008],[360,-0.0107],[362,-0.0239],[364,0.0017],[366,0.1299],[368,0.0498]]
       }
       ball_l_ik_ctrl.Scale.X {type=MovieSceneScriptingFloatChannel, default=1, const [1,140]}
       ball_l_ik_ctrl.Scale.Y {type=MovieSceneScriptingFloatChannel, default=1, const [1,140]}
       ball_l_ik_ctrl.Scale.Z {type=MovieSceneScriptingFloatChannel, default=1, const [1,140]}
       foot_bk1_l_ctrl.Location.X {type=MovieSceneScriptingFloatChannel, default=0, const [0,140]}
       foot_bk1_l_ctrl.Location.Y {type=MovieSceneScriptingFloatChannel, default=0, const [0,140]}
       foot_bk1_l_ctrl.Location.Z {type=MovieSceneScriptingFloatChannel, default=0, const [0,140]}
       foot_bk1_l_ctrl.Rotation.X {type=MovieSceneScriptingFloatChannel, default=0, const [0,140]}
       foot_bk1_l_ctrl.Rotation.Y {type=MovieSceneScriptingFloatChannel, default=0, const [0,140]}
       foot_bk1_l_ctrl.Rotation.Z {type=MovieSceneScriptingFloatChannel, default=0, const [0,140]}
       foot_bk1_l_ctrl.Scale.X {type=MovieSceneScriptingFloatChannel, default=1, const [1,140]}
       foot_bk1_l_ctrl.Scale.Y {type=MovieSceneScriptingFloatChannel, default=1, const [1,140]}
       foot_bk1_l_ctrl.Scale.Z {type=MovieSceneScriptingFloatChannel, default=1, const [1,140]}
       foot_roll_l_ctrl.Location.X {type=MovieSceneScriptingFloatChannel, default=0, const [0,140]}
       foot_roll_l_ctrl.Location.Y {type=MovieSceneScriptingFloatChannel, default=0, const [0,140]}
       foot_roll_l_ctrl.Location.Z {type=MovieSceneScriptingFloatChannel, default=0, const [0,140]}
       foot_roll_l_ctrl.Rotation.X {type=MovieSceneScriptingFloatChannel, default=0, const [0,140]}
       foot_roll_l_ctrl.Rotation.Y {type=MovieSceneScriptingFloatChannel, default=0, const [0,140]}
       foot_roll_l_ctrl.Rotation.Z {type=MovieSceneScriptingFloatChannel, default=0, const [0,140]}
       foot_roll_l_ctrl.Scale.X {type=MovieSceneScriptingFloatChannel, default=1, const [1,140]}
       foot_roll_l_ctrl.Scale.Y {type=MovieSceneScriptingFloatChannel, default=1, const [1,140]}
       foot_roll_l_ctrl.Scale.Z {type=MovieSceneScriptingFloatChannel, default=1, const [1,140]}
       roll_blend_l_ctrl {type=MovieSceneScriptingFloatChannel, default=1, const [1,140]}
       tip_l_ctrl.Location.X {type=MovieSceneScriptingFloatChannel, default=0, const [0,140]}
       tip_l_ctrl.Location.Y {type=MovieSceneScriptingFloatChannel, default=0, const [0,140]}
       tip_l_ctrl.Location.Z {type=MovieSceneScriptingFloatChannel, default=0, const [0,140]}
       tip_l_ctrl.Rotation.X {type=MovieSceneScriptingFloatChannel, default=0, const [0,140]}
       tip_l_ctrl.Rotation.Y {type=MovieSceneScriptingFloatChannel, default=0, const [0,140]}
       tip_l_ctrl.Rotation.Z {type=MovieSceneScriptingFloatChannel, default=0, const [0,140]}
       tip_l_ctrl.Scale.X {type=MovieSceneScriptingFloatChannel, default=1, const [1,140]}
       tip_l_ctrl.Scale.Y {type=MovieSceneScriptingFloatChannel, default=1, const [1,140]}
       tip_l_ctrl.Scale.Z {type=MovieSceneScriptingFloatChannel, default=1, const [1,140]}
       heel_l_ctrl.Location.X {type=MovieSceneScriptingFloatChannel, default=0, const [0,140]}
       heel_l_ctrl.Location.Y {type=MovieSceneScriptingFloatChannel, default=0, const [0,140]}
       heel_l_ctrl.Location.Z {type=MovieSceneScriptingFloatChannel, default=0, const [0,140]}
       heel_l_ctrl.Rotation.X {type=MovieSceneScriptingFloatChannel, default=0, const [0,140]}
       heel_l_ctrl.Rotation.Y {type=MovieSceneScriptingFloatChannel, default=0, const [0,140]}
       heel_l_ctrl.Rotation.Z {type=MovieSceneScriptingFloatChannel, default=0, const [0,140]}
       heel_l_ctrl.Scale.X {type=MovieSceneScriptingFloatChannel, default=1, const [1,140]}
       heel_l_ctrl.Scale.Y {type=MovieSceneScriptingFloatChannel, default=1, const [1,140]}
       heel_l_ctrl.Scale.Z {type=MovieSceneScriptingFloatChannel, default=1, const [1,140]}
       leg_l_softness {type=MovieSceneScriptingFloatChannel, default=0, const [0,140]}
       leg_l_pv_ik_ctrl.Location.X {
        type=MovieSceneScriptingFloatChannel
        default=-10.5581
        thinned_from=140
        keys [[0,22.0566],[92,22.0566],[94,22.7592],[98,24.4586],[100,25.1533],[102,25.4142],[104,25.7464],[106,25.692],[108,25.0384],[110,23.4916],[112,22.3707],[114,20.9749],[116,20.0823],[118,19.5185],[120,18.7899],[122,17.4113],[124,16.4873],[126,15.636],[130,14.2948],[132,13.9743],[134,13.5341],[136,12.8268],[138,12.4619],[140,12.5228],[142,12.0335],[144,12.1419],[146,11.9226],[148,11.3354],[150,10.8036],[152,10.2083],[154,9.5109],[156,8.8392],[158,8.2249],[160,7.7429],[162,7.3487],[164,7.2364],[166,7.2056],[170,6.5758],[172,6.4204],[174,6.0371],[176,5.5238],[178,4.64],[180,3.8828],[182,3.3894],[184,3.1253],[186,2.981],[188,2.8668],[190,2.8276],[192,3.0383],[194,3.5125],[196,3.9083],[198,4.0461],[200,4.3582],[202,4.5056],[204,4.6011],[206,4.6253],[208,4.6648],[210,4.7274],[212,4.7364],[216,4.5052],[218,4.6619],[220,4.8372],[222,4.9768],[224,5.0346],[226,5.2441],[228,5.5115],[230,5.6541],[232,5.6558],[234,5.63],[236,5.7151],[238,5.869],[240,5.9657],[242,6.0196],[244,6.1242],[246,6.3261],[248,6.4],[250,6.4152],[252,6.4869],[254,6.5933],[256,6.7779],[258,6.8076],[262,6.6832],[264,6.6773],[266,6.7316],[268,6.883],[272,6.8911],[280,7.3052],[282,7.3381],[284,7.2919],[286,7.1868],[288,7.16],[290,7.1813],[292,7.1643],[296,7.2629],[298,7.3232],[300,7.414],[304,7.647],[306,7.8278],[308,8.0407],[310,8.1632],[312,8.2553],[314,8.2513],[316,8.3536],[318,8.5228],[320,8.8066],[322,8.8815],[324,9.1539],[326,9.3295],[328,9.2457],[330,8.916],[332,8.6408],[334,8.199],[336,6.753],[338,6.3127],[340,5.5205],[342,6.5697],[344,7.9262],[346,7.0033],[348,5.6299],[350,4.4965],[352,3.9716],[354,4.1507],[356,4.3897],[358,5.6072],[360,7.3034],[362,10.3113],[364,10.2859],[366,5.9618],[368,-10.5581]]
       }
       leg_l_pv_ik_ctrl.Location.Y {
        type=MovieSceneScriptingFloatChannel
        default=86.2495
        thinned_from=140
        keys [[0,-1071.1206],[94,-1070.7666],[96,-995.986],[98,-900.1468],[100,-791.6312],[102,-678.2814],[104,-567.9058],[106,-469.356],[108,-390.6118],[110,-340.7153],[112,-308.5934],[114,-279.2694],[116,-252.3172],[118,-228.0414],[120,-206.3822],[122,-187.6575],[124,-170.9665],[126,-156.9427],[128,-144.1672],[130,-132.4632],[132,-121.8478],[134,-112.0103],[136,-103.3447],[138,-95.3422],[140,-87.5749],[142,-80.2309],[144,-73.2835],[146,-66.8607],[148,-60.7404],[150,-54.8166],[152,-49.3171],[154,-44.367],[156,-39.8639],[158,-35.746],[160,-32.1115],[162,-28.9393],[164,-26.0805],[166,-23.4142],[168,-20.8254],[170,-18.4481],[172,-16.3295],[174,-14.4221],[176,-12.6334],[178,-11.0447],[180,-9.573],[182,-8.1846],[184,-6.8367],[186,-5.6219],[188,-4.511],[190,-3.5856],[192,-2.7762],[194,-2.0817],[196,-1.5815],[198,-1.2952],[200,-1.1151],[202,-0.9938],[204,-0.9621],[206,-0.9865],[210,-0.9967],[212,-1.0068],[214,-0.9959],[220,-0.8743],[222,-0.8415],[224,-0.8144],[226,-0.7805],[228,-0.722],[230,-0.6556],[232,-0.5527],[234,-0.5033],[236,-0.4834],[238,-0.4702],[240,-0.4707],[242,-0.4637],[244,-0.4327],[246,-0.3548],[248,-0.2901],[250,-0.2209],[252,-0.1308],[254,-0.0756],[256,0.011],[258,0.0257],[260,0.0203],[262,-0.0236],[264,-0.0939],[266,-0.122],[268,-0.2078],[270,-0.3963],[272,-0.5922],[274,-0.691],[276,-0.7267],[278,-0.7063],[280,-0.6312],[282,-0.6033],[284,-0.6231],[286,-0.6865],[288,-0.7518],[290,-0.8741],[292,-1.0097],[294,-1.1126],[296,-1.1414],[298,-1.2089],[300,-1.2123],[302,-1.1023],[304,-0.9366],[306,-0.6855],[308,-0.3332],[310,-0.1262],[312,0.0593],[314,0.1692],[316,0.2668],[318,0.318],[320,0.3931],[322,0.4086],[324,0.4522],[326,0.432],[328,0.3213],[330,0.1975],[332,0.3722],[334,1.3039],[336,5.0283],[338,8.721],[340,11.6923],[342,14.0194],[344,15.4376],[346,16.7556],[348,17.8867],[350,18.582],[352,18.9505],[354,19.2413],[356,19.3956],[358,19.3484],[360,18.9759],[362,17.494],[364,13.7606],[366,14.9985],[368,86.2495]]
       }
       leg_l_pv_ik_ctrl.Location.Z {
        type=MovieSceneScriptingFloatChannel
        default=67.4649
        thinned_from=140
        keys [[0,5.9046],[92,5.9046],[96,5.1406],[98,4.9716],[100,5.0721],[102,4.4082],[104,3.0615],[106,2.1315],[108,-0.1025],[110,-1.3067],[112,-3.7054],[114,-5.4313],[116,-6.8664],[118,-7.6477],[120,-8.1552],[122,-8.229],[124,-8.3956],[126,-7.9858],[128,-7.7785],[130,-7.7633],[132,-7.7984],[134,-7.9613],[136,-8.1862],[138,-8.3319],[140,-8.2931],[142,-8.1157],[144,-7.8198],[146,-7.4571],[148,-7.0656],[150,-6.5888],[152,-6.0938],[154,-5.6616],[156,-5.2773],[158,-4.9461],[160,-4.7296],[162,-4.6353],[164,-4.6271],[166,-4.6473],[168,-4.6194],[170,-4.6185],[174,-4.7179],[176,-4.7529],[180,-4.7873],[182,-4.7937],[184,-4.765],[186,-4.7608],[188,-4.7305],[190,-4.7397],[192,-4.7592],[194,-4.7898],[196,-4.8327],[200,-5.0036],[202,-5.0538],[206,-5.1242],[208,-5.0939],[210,-5.0297],[212,-4.9314],[214,-4.7791],[218,-4.4058],[220,-4.1934],[222,-3.9691],[224,-3.7271],[226,-3.4991],[228,-3.2583],[230,-3.0269],[232,-2.7191],[234,-2.4391],[238,-1.8921],[240,-1.6088],[242,-1.3085],[244,-1.0011],[246,-0.6718],[250,-0.0355],[252,0.2783],[254,0.5557],[256,0.8212],[258,1.0513],[260,1.2584],[262,1.425],[264,1.5628],[266,1.6966],[268,1.7996],[270,1.808],[272,1.78],[274,1.7677],[276,1.7642],[278,1.777],[280,1.8174],[282,1.8225],[284,1.7957],[286,1.7216],[288,1.6335],[290,1.4997],[292,1.3514],[294,1.2202],[296,1.1202],[298,1.0013],[300,0.9049],[302,0.8652],[304,0.85],[306,0.8601],[308,0.9168],[310,0.8965],[312,0.8588],[316,0.7001],[318,0.6107],[320,0.5641],[322,0.5103],[324,0.5039],[326,0.4769],[328,0.4087],[330,0.356],[332,0.5478],[334,1.446],[336,5.0952],[338,9.1985],[340,13.2219],[342,16.6303],[344,18.6857],[346,20.9516],[348,23.5362],[350,25.7555],[352,27.3953],[354,28.7883],[356,29.6599],[358,29.6602],[360,25.2204],[362,16.0963],[364,-1.1767],[366,4.2133],[368,67.4649]]
       }
       leg_l_pv_ik_ctrl.Rotation.X {
        type=MovieSceneScriptingFloatChannel
        default=54.9411
        thinned_from=140
        keys [[0,94.5453],[92,94.5453],[96,94.084],[98,94.0282],[100,94.1633],[102,93.3941],[106,90.1461],[108,87.0561],[110,84.7309],[112,80.882],[114,77.3021],[116,73.5138],[120,66.8851],[122,64.2404],[124,61.0645],[126,59.1917],[128,56.8975],[132,50.8794],[134,47.3753],[136,44.0373],[140,37.8998],[142,34.9007],[144,32.2485],[150,25.3727],[152,23.2402],[154,21.3826],[156,19.7154],[158,18.2361],[160,17.1008],[162,16.2631],[164,15.5979],[166,14.9746],[168,14.2175],[170,13.5403],[174,12.4553],[176,11.8782],[180,10.6641],[182,10.0786],[184,9.4592],[188,8.4267],[190,8.0307],[192,7.708],[194,7.4691],[196,7.3169],[198,7.2597],[200,7.2555],[202,7.2322],[204,7.2522],[206,7.3157],[210,7.3774],[212,7.3954],[214,7.3542],[216,7.2746],[224,7.2149],[228,7.3049],[230,7.3906],[232,7.3571],[234,7.4045],[238,7.565],[240,7.6129],[242,7.6273],[244,7.6216],[248,7.5384],[250,7.4741],[252,7.3791],[256,7.2766],[260,7.2787],[264,7.3729],[266,7.3881],[268,7.4357],[272,7.7784],[274,7.8504],[276,7.851],[278,7.79],[280,7.6653],[282,7.5793],[284,7.5378],[286,7.572],[288,7.6305],[292,7.9176],[294,8.0085],[296,8.0299],[298,8.069],[300,8.0506],[302,7.9079],[304,7.7166],[306,7.4814],[308,7.182],[310,7.0399],[312,6.9349],[314,6.9258],[316,6.9536],[318,7.0226],[320,7.0418],[322,7.0458],[324,7.0095],[326,7.0121],[328,7.0791],[330,7.0981],[332,6.6736],[334,4.9901],[336,-1.7298],[338,-9.1396],[340,-16.5003],[342,-22.6306],[344,-26.3446],[346,-30.653],[348,-34.8752],[350,-38.3664],[352,-40.778],[354,-42.4932],[356,-43.3865],[358,-43.2769],[360,-39.4998],[362,-31.7857],[364,-8.7296],[366,26.7333],[368,54.9411]]
       }
       leg_l_pv_ik_ctrl.Rotation.Y {
        type=MovieSceneScriptingFloatChannel
        default=-3.2953
        thinned_from=140
        keys [[0,-9.7326],[92,-9.7326],[94,-9.0988],[96,-8.1404],[98,-7.0551],[100,-5.9853],[102,-5.4336],[104,-4.5989],[106,-3.3711],[108,-3.0584],[110,-3.421],[112,-3.8165],[114,-4.3993],[116,-4.5152],[118,-4.088],[120,-3.8235],[122,-4.3843],[124,-4.6447],[126,-4.982],[128,-5.1251],[130,-5.1088],[132,-4.6642],[134,-4.1937],[136,-4.1211],[138,-3.8502],[140,-3.222],[142,-3.1366],[144,-2.5426],[146,-2.2062],[148,-2.1972],[150,-2.0963],[152,-2.0387],[154,-2.161],[156,-2.3624],[158,-2.5509],[160,-2.6388],[162,-2.7079],[164,-2.6511],[166,-2.5492],[168,-2.5687],[170,-2.6011],[172,-2.6107],[174,-2.7867],[176,-2.9883],[178,-3.3757],[180,-3.7324],[182,-3.9571],[184,-4.0765],[188,-4.1872],[190,-4.2267],[192,-4.1704],[194,-4.0019],[196,-3.888],[198,-3.8946],[200,-3.8837],[202,-3.8938],[204,-3.9253],[206,-3.9732],[208,-4.0124],[210,-4.0317],[212,-4.0725],[214,-4.1436],[216,-4.2284],[218,-4.2026],[220,-4.1673],[222,-4.1519],[224,-4.1609],[226,-4.0939],[228,-3.9835],[230,-3.8953],[232,-3.8572],[234,-3.8306],[238,-3.7011],[240,-3.676],[242,-3.6587],[244,-3.6239],[246,-3.5318],[248,-3.4928],[252,-3.4593],[254,-3.4265],[256,-3.364],[258,-3.3548],[260,-3.3767],[264,-3.4422],[266,-3.4455],[268,-3.4681],[270,-3.5445],[272,-3.6097],[274,-3.6245],[276,-3.6171],[278,-3.5938],[280,-3.5577],[282,-3.57],[284,-3.6108],[286,-3.6731],[290,-3.7434],[292,-3.8041],[294,-3.8464],[296,-3.8485],[298,-3.8701],[300,-3.8518],[302,-3.7968],[304,-3.7193],[306,-3.5558],[308,-3.3141],[310,-3.1466],[312,-2.9871],[314,-2.8828],[316,-2.7185],[318,-2.5606],[320,-2.3404],[322,-2.2744],[324,-2.1017],[326,-2.0063],[328,-2.0692],[330,-2.2366],[334,-2.5376],[336,-2.739],[338,-2.7214],[340,-2.9608],[342,-2.5971],[344,-2.4604],[346,-2.6994],[348,-3.0921],[350,-2.8532],[352,-2.4014],[354,-1.9443],[356,-1.1762],[358,-0.8872],[360,-0.7659],[362,-1.3516],[364,0.2481],[366,-4.0627],[368,-3.2953]]
       }
       leg_l_pv_ik_ctrl.Rotation.Z {
        type=MovieSceneScriptingFloatChannel
        default=0.7176
        thinned_from=140
        keys [[0,8.6415],[92,8.6415],[94,9.1727],[96,9.6669],[98,9.9902],[100,10.1356],[102,10.3144],[104,10.6294],[106,10.1617],[108,9.8511],[110,8.9403],[112,8.9799],[114,8.933],[116,9.3001],[118,9.4724],[120,9.5378],[122,9.5697],[124,10.022],[126,10.3261],[128,10.8231],[130,11.3662],[132,11.6068],[134,11.489],[136,11.0339],[138,10.549],[140,9.8651],[142,9.6619],[144,9.3429],[146,9.4975],[148,9.8877],[150,10.1956],[152,10.5298],[154,10.8308],[160,11.4154],[162,11.5849],[164,11.5299],[166,11.4232],[170,11.7998],[172,11.7989],[174,12.0119],[176,12.3866],[178,13.1756],[180,13.8187],[182,14.1896],[184,14.3221],[188,14.3217],[190,14.2496],[192,13.9131],[194,13.2946],[196,12.7723],[198,12.5566],[200,12.1591],[202,11.9755],[204,11.8681],[206,11.8532],[210,11.7278],[212,11.709],[216,11.9686],[220,11.6404],[222,11.5156],[224,11.4842],[226,11.3019],[228,11.0485],[230,10.9763],[232,11.0106],[234,11.0888],[236,11.0385],[238,10.8946],[240,10.7937],[242,10.7358],[244,10.6124],[246,10.3906],[248,10.3182],[250,10.3199],[252,10.2379],[254,10.131],[256,9.9215],[258,9.8881],[262,10.0329],[264,10.0358],[266,9.9709],[268,9.7868],[272,9.8542],[276,9.6948],[280,9.4918],[282,9.4547],[284,9.4968],[286,9.6013],[288,9.6224],[290,9.5997],[292,9.6259],[296,9.5126],[298,9.4329],[300,9.3318],[304,9.0396],[306,8.8428],[308,8.6104],[310,8.4737],[312,8.3663],[314,8.3776],[316,8.2838],[318,8.1184],[320,7.8274],[322,7.7482],[324,7.4478],[326,7.2555],[328,7.3208],[330,7.6543],[332,7.8205],[334,8.0399],[336,8.7987],[338,8.4432],[340,8.5127],[342,7.0268],[344,5.3317],[346,5.8998],[348,6.8959],[350,7.809],[352,8.1955],[354,8.2162],[356,8.3368],[358,7.768],[360,7.5275],[362,6.5901],[364,8.8306],[366,8.4552],[368,0.7176]]
       }
       leg_l_pv_ik_ctrl.Scale.X {type=MovieSceneScriptingFloatChannel, default=1, const [1,140]}
       leg_l_pv_ik_ctrl.Scale.Y {type=MovieSceneScriptingFloatChannel, default=1, const [1,140]}
       leg_l_pv_ik_ctrl.Scale.Z {type=MovieSceneScriptingFloatChannel, default=1, const [1,140]}
       thigh_r_fk_ctrl.Location.X {type=MovieSceneScriptingFloatChannel, default=0, thinned_from=140, keys [[0,0],[368,0]]}
       thigh_r_fk_ctrl.Location.Y {type=MovieSceneScriptingFloatChannel, default=0, thinned_from=140, keys [[0,0],[368,0]]}
       thigh_r_fk_ctrl.Location.Z {type=MovieSceneScriptingFloatChannel, default=0, thinned_from=140, keys [[0,0],[368,0]]}
       thigh_r_fk_ctrl.Rotation.X {
        type=MovieSceneScriptingFloatChannel
        default=0
        thinned_from=140
        keys [[0,-9.7405],[92,-9.7405],[94,-9.6381],[96,-9.6818],[98,-9.901],[100,-10.3295],[102,-10.3794],[104,-10.2317],[106,-10.3698],[108,-9.9334],[110,-9.8723],[112,-9.025],[114,-8.2343],[116,-7.2077],[118,-6.6969],[120,-6.1654],[122,-5.6159],[124,-5.0498],[126,-4.4686],[128,-3.8738],[130,-3.2668],[132,-2.6491],[134,-2.0222],[136,-1.3876],[138,-0.7467],[140,-0.1009],[142,0.5481],[150,3.1483],[154,4.4301],[158,5.6841],[162,6.8984],[166,8.0614],[168,8.6199],[170,9.1612],[172,9.6838],[174,10.1861],[176,10.6668],[178,11.1243],[180,11.5572],[182,11.964],[184,12.3433],[186,12.6935],[188,13.0133],[190,13.301],[192,13.5554],[194,13.7748],[196,13.9578],[198,14.1028],[200,14.2083],[202,14.2726],[204,14.2944],[206,14.1935],[208,14.2274],[212,14.3471],[214,14.3646],[218,14.3275],[220,14.3886],[222,14.5327],[224,14.637],[226,14.6968],[228,14.704],[230,14.6695],[232,14.7329],[234,14.8451],[236,14.921],[242,14.8712],[246,14.896],[250,14.8436],[252,14.9147],[256,14.977],[260,14.9415],[264,14.7784],[266,14.767],[268,14.7146],[272,14.4137],[274,14.3574],[276,14.3852],[282,14.627],[284,14.6289],[286,14.526],[290,14.2382],[292,14.144],[294,14.1176],[296,14.1305],[300,14.2383],[302,14.3213],[304,14.4468],[306,14.6174],[308,14.8989],[312,15.3004],[316,15.4518],[322,15.465],[324,15.4338],[326,15.3204],[328,15.0685],[330,14.768],[332,14.3743],[334,14.0853],[338,13.4701],[340,13.2426],[342,13.1161],[344,13.1221],[346,13.3212],[348,13.681],[350,14.0933],[352,14.4501],[354,14.6431],[356,13.7578],[358,11.4974],[360,8.6924],[362,6.1735],[364,2.4655],[366,0.8358],[368,0]]
       }
       thigh_r_fk_ctrl.Rotation.Y {
        type=MovieSceneScriptingFloatChannel
        default=-8.0673
        thinned_from=140
        keys [[0,0.8739],[92,0.8739],[94,0.804],[96,0.6785],[98,0.4687],[100,0.1011],[102,-0.2496],[104,-0.9786],[106,-1.8268],[108,-2.3854],[110,-3.3298],[112,-3.915],[114,-4.3627],[116,-4.4168],[120,-4.1842],[124,-3.933],[130,-3.5273],[138,-2.9476],[154,-1.7421],[158,-1.448],[162,-1.1627],[164,-1.0242],[166,-0.889],[168,-0.7573],[170,-0.6297],[172,-0.5065],[174,-0.3879],[176,-0.2745],[178,-0.1664],[180,-0.0642],[182,0.0319],[184,0.1214],[186,0.2041],[188,0.2795],[190,0.3473],[192,0.4071],[194,0.4586],[196,0.5014],[198,0.5352],[200,0.5598],[202,0.5748],[204,0.5799],[206,0.5753],[208,0.5831],[210,0.5975],[212,0.6053],[214,0.598],[216,0.5296],[218,0.5279],[220,0.6179],[222,0.7811],[224,1.0023],[226,1.3055],[228,1.5869],[230,1.9492],[232,2.0241],[238,2.2788],[240,2.3124],[244,2.3154],[246,2.2566],[248,2.1617],[250,2.0841],[252,2.0721],[254,2.1285],[256,2.2238],[258,2.2499],[260,2.1661],[262,2.0975],[264,2.1037],[266,2.1373],[268,2.2512],[270,2.302],[272,2.2541],[274,2.2354],[276,2.2629],[278,2.2991],[280,2.3635],[282,2.4409],[284,2.5013],[286,2.4577],[290,2.4208],[292,2.4528],[294,2.493],[296,2.5017],[298,2.5323],[300,2.5383],[304,2.5322],[306,2.5216],[310,2.6188],[314,2.766],[316,2.9035],[318,3.086],[320,3.2965],[322,3.4157],[324,3.7135],[326,4.0568],[328,4.3831],[330,4.5559],[332,4.503],[334,3.9058],[336,2.5353],[338,0.8379],[340,-0.7535],[342,-1.866],[344,-2.1268],[346,-1.8249],[348,-1.5118],[350,-1.2083],[352,-0.9353],[354,-0.7138],[356,-0.1378],[358,0.9528],[360,1.814],[362,2.0473],[364,-2.1344],[366,-8.2755],[368,-8.0673]]
       }
       thigh_r_fk_ctrl.Rotation.Z {
        type=MovieSceneScriptingFloatChannel
        default=0
        thinned_from=140
        keys [[0,9.0095],[92,9.0095],[94,9.2203],[96,9.6276],[98,10.1106],[100,10.96],[102,11.3359],[104,11.9495],[106,13.7233],[108,14.0404],[110,15.3715],[112,15.0694],[114,14.6597],[116,13.6107],[118,13.0917],[120,12.2456],[122,11.3646],[124,10.4512],[126,9.508],[128,8.5377],[130,7.5429],[132,6.5261],[134,5.4899],[136,4.4371],[138,3.3701],[140,2.2915],[142,1.2041],[144,0.1103],[150,-3.183],[154,-5.3624],[158,-7.5047],[162,-9.589],[164,-10.6029],[166,-11.5945],[168,-12.5611],[170,-13.5003],[172,-14.4093],[174,-15.2857],[176,-16.1267],[178,-16.9298],[180,-17.6923],[182,-18.4118],[184,-19.0855],[186,-19.7108],[188,-20.2853],[190,-20.8061],[192,-21.2709],[194,-21.6769],[196,-22.0216],[198,-22.2979],[200,-22.4988],[202,-22.6214],[206,-22.6618],[208,-22.6099],[212,-22.4266],[220,-22.4509],[224,-22.3507],[228,-22.1409],[230,-21.9435],[234,-21.6857],[236,-21.6069],[248,-21.4618],[254,-21.5657],[264,-21.3788],[266,-21.3734],[272,-21.5207],[274,-21.4884],[278,-21.274],[280,-21.2373],[282,-21.2495],[290,-21.4802],[310,-21.5824],[312,-21.5617],[318,-21.2707],[320,-21.1122],[324,-20.9173],[330,-20.8204],[332,-20.9143],[334,-21.6586],[336,-23.314],[338,-26.121],[340,-30.1382],[342,-34.6276],[344,-38.8509],[346,-42.854],[348,-46.992],[350,-50.9912],[352,-54.5778],[354,-57.4777],[356,-58.3439],[358,-56.2198],[360,-50.7822],[362,-43.0376],[364,-15.829],[366,-8.7553],[368,0]]
       }
       thigh_r_fk_ctrl.Scale.X {type=MovieSceneScriptingFloatChannel, default=1, const [1,140]}
       thigh_r_fk_ctrl.Scale.Y {type=MovieSceneScriptingFloatChannel, default=1, const [1,140]}
       thigh_r_fk_ctrl.Scale.Z {type=MovieSceneScriptingFloatChannel, default=1, const [1,140]}
       calf_r_fk_ctrl.Location.X {
        type=MovieSceneScriptingFloatChannel
        default=0
        thinned_from=140
        keys [[0,0],[94,0],[98,0.0001],[102,0],[108,0.0005],[110,-0.0003],[112,0.0005],[114,0.0008],[116,0.0004],[222,0.0002],[224,0.0004],[226,0.0002],[368,0]]
       }
       calf_r_fk_ctrl.Location.Y {
        type=MovieSceneScriptingFloatChannel
        default=0
        thinned_from=140
        keys [[0,-0.0002],[92,-0.0002],[94,0.0003],[96,0.0014],[98,0.0014],[100,-0.0005],[102,-0.0003],[104,0.0014],[106,0.0018],[108,0.001],[110,-0.0103],[112,0.0005],[114,0],[204,0.0008],[206,0.0003],[208,-0.0004],[210,-0.0009],[212,-0.001],[214,0.0006],[216,0.0016],[218,0.0016],[220,0.0006],[222,0],[224,0.002],[226,-0.0006],[228,0.0005],[230,0.0003],[232,-0.0009],[234,-0.0007],[236,0.0004],[238,0.0003],[240,-0.0004],[242,0.0008],[244,0.0001],[246,0.0003],[250,-0.0008],[252,0.0003],[254,0.0009],[256,-0.0003],[258,0.0009],[260,0.0001],[262,0.0006],[264,-0.0004],[266,-0.0005],[268,-0.0001],[270,0.0008],[272,0.0011],[274,0.0011],[276,-0.0001],[278,-0.0001],[280,-0.0008],[282,0.0004],[290,0.0001],[292,0.0008],[294,-0.0001],[296,-0.0002],[298,0.0006],[300,0.0006],[302,-0.0002],[304,0.0002],[306,-0.0001],[308,0.0002],[310,-0.0006],[312,-0.0009],[314,0.0001],[316,-0.0004],[320,-0.0004],[322,-0.0001],[324,-0.0007],[330,0],[368,0]]
       }
       calf_r_fk_ctrl.Location.Z {
        type=MovieSceneScriptingFloatChannel
        default=0
        thinned_from=140
        keys [[0,0],[94,0],[98,0.0002],[100,-0.0001],[104,0],[108,-0.0002],[110,0.0021],[112,-0.0004],[116,-0.0001],[204,0.0004],[212,-0.0005],[214,0.0004],[216,0.0008],[218,0.0008],[220,0.0003],[222,0],[224,0.0011],[226,-0.0003],[228,0.0003],[230,0.0002],[232,-0.0005],[234,-0.0004],[236,0.0003],[238,0.0002],[240,-0.0002],[242,0.0005],[244,0.0001],[246,0.0002],[250,-0.0005],[252,0.0002],[254,0.0006],[256,-0.0002],[258,0.0006],[260,0.0001],[262,0.0004],[264,-0.0002],[266,-0.0003],[268,-0.0001],[270,0.0005],[274,0.0007],[276,0],[278,0],[280,-0.0005],[282,0.0003],[290,0.0001],[292,0.0006],[294,0],[296,-0.0001],[298,0.0004],[300,0.0004],[302,-0.0001],[304,0.0001],[306,-0.0001],[308,0.0001],[310,-0.0004],[312,-0.0006],[314,0.0001],[316,-0.0003],[320,-0.0003],[322,0],[324,-0.0005],[326,-0.0002],[368,0]]
       }
       calf_r_fk_ctrl.Rotation.X {
        type=MovieSceneScriptingFloatChannel
        default=0
        thinned_from=140
        keys [[0,-0.0711],[92,-0.0711],[94,-0.0682],[96,-0.0507],[98,-0.049],[100,-0.0544],[102,-0.43],[104,-1.3974],[106,-2.7419],[108,-4.2463],[110,-5.697],[112,-6.8224],[114,-7.4502],[162,-5.3263],[172,-4.9366],[180,-4.6682],[186,-4.4987],[192,-4.3611],[196,-4.289],[200,-4.2367],[204,-4.2082],[220,-4.2015],[224,-4.1852],[226,-4.1888],[230,-4.2211],[234,-4.2103],[254,-4.2255],[258,-4.2217],[262,-4.2312],[266,-4.2261],[270,-4.2376],[274,-4.2311],[282,-4.2472],[284,-4.2372],[290,-4.2505],[304,-4.2391],[306,-4.2432],[314,-4.2135],[316,-4.215],[324,-4.2648],[332,-4.3564],[336,-4.421],[338,-4.3722],[340,-4.173],[342,-3.8512],[344,-3.4342],[346,-2.9585],[348,-2.4642],[350,-1.9843],[352,-1.5517],[354,-1.1995],[356,-0.6134],[358,0.2487],[360,0.9967],[362,1.2971],[364,0.7304],[366,0.3525],[368,0]]
       }
       calf_r_fk_ctrl.Rotation.Y {
        type=MovieSceneScriptingFloatChannel
        default=0
        thinned_from=140
        keys [[0,0.0139],[92,0.0139],[94,0.0292],[96,0.055],[98,0.0731],[100,0.0685],[102,0.0984],[104,0.1813],[106,0.2897],[108,0.4889],[110,0.7334],[112,1.1059],[114,1.5688],[116,2.0899],[118,2.189],[124,2.5082],[158,4.4633],[166,4.8852],[174,5.2631],[180,5.5092],[186,5.7166],[192,5.8794],[196,5.9603],[200,6.0158],[204,6.0433],[222,6.0284],[226,5.9995],[230,5.9427],[274,5.92],[282,5.8836],[306,5.8796],[312,5.8977],[314,5.8923],[318,5.8502],[320,5.8137],[322,5.794],[328,5.6574],[332,5.5755],[334,5.6126],[336,5.7059],[338,5.8451],[342,6.1911],[344,6.3376],[352,6.8722],[354,6.9907],[356,5.8765],[360,0.6992],[362,-0.4212],[364,-0.0242],[366,-0.0018],[368,0]]
       }
       calf_r_fk_ctrl.Rotation.Z {
        type=MovieSceneScriptingFloatChannel
        default=0
        thinned_from=140
        keys [[0,0.3405],[92,0.3405],[94,0.4903],[96,0.745],[98,0.9314],[100,0.9589],[102,2.3871],[104,6.1947],[106,11.6629],[108,18.7472],[110,26.6849],[112,35.5282],[114,44.5389],[116,52.8028],[118,54.5802],[132,65.1198],[138,69.1748],[140,70.3432],[142,71.3441],[144,72.1874],[148,73.442],[152,74.1877],[156,74.505],[162,74.3546],[184,71.514],[190,71.3153],[194,71.589],[200,72.4341],[206,72.7549],[224,72.5125],[230,71.7791],[274,71.5742],[282,71.2356],[314,71.2774],[318,70.854],[320,70.4885],[322,70.2851],[328,68.9147],[332,68.1161],[334,68.4326],[336,69.2031],[338,71.0729],[340,74.4684],[342,78.8569],[344,83.7057],[348,93.9219],[350,98.8891],[352,103.3616],[354,107.0387],[356,109.6465],[358,110.8223],[360,109.7575],[362,106.9318],[364,52.5318],[366,22.6533],[368,0]]
       }
       calf_r_fk_ctrl.Scale.X {type=MovieSceneScriptingFloatChannel, default=1, const [1,140]}
       calf_r_fk_ctrl.Scale.Y {type=MovieSceneScriptingFloatChannel, default=1, const [1,140]}
       calf_r_fk_ctrl.Scale.Z {type=MovieSceneScriptingFloatChannel, default=1, const [1,140]}
       foot_r_fk_ctrl.Location.X {
        type=MovieSceneScriptingFloatChannel
        default=0
        thinned_from=140
        keys [[0,0],[104,0.0002],[108,0.0004],[110,0.0008],[112,0.0006],[114,0.0011],[116,0.0005],[330,0.0001],[332,0.0009],[354,0.0009],[362,0],[368,0]]
       }
       foot_r_fk_ctrl.Location.Y {
        type=MovieSceneScriptingFloatChannel
        default=0
        thinned_from=140
        keys [[0,0],[354,0],[356,-0.0052],[360,-0.0278],[362,-0.0329],[364,-0.0193],[368,0]]
       }
       foot_r_fk_ctrl.Location.Z {
        type=MovieSceneScriptingFloatChannel
        default=0
        thinned_from=140
        keys [[0,0],[354,0],[356,-0.1418],[360,-0.7657],[362,-0.9075],[364,-0.5334],[368,0]]
       }
       foot_r_fk_ctrl.Rotation.X {
        type=MovieSceneScriptingFloatChannel
        default=0
        thinned_from=140
        keys [[0,0.2247],[92,0.2247],[94,0.1396],[96,0.0647],[98,0.3058],[100,0.2743],[102,0.4606],[104,1.1041],[106,2.1454],[108,2.7308],[110,3.4845],[112,4.8752],[114,6.3697],[116,7.0432],[120,7.021],[124,6.9571],[128,6.856],[132,6.722],[138,6.469],[144,6.1666],[170,4.6469],[176,4.3309],[180,4.1439],[184,3.981],[188,3.8465],[190,3.7912],[192,3.7447],[194,3.7075],[196,3.6801],[200,3.6459],[204,3.634],[206,3.6455],[210,3.6138],[212,3.6983],[214,3.7336],[216,3.6719],[218,3.5813],[222,3.3355],[224,3.1798],[228,2.4208],[230,1.8061],[232,1.6738],[234,1.6138],[236,1.5114],[238,1.3008],[240,1.215],[242,1.2434],[244,1.2765],[246,1.2554],[248,1.2559],[250,1.2619],[252,1.2568],[254,1.221],[256,1.1295],[258,1.0612],[260,1.0015],[262,1.0067],[264,0.9462],[266,0.8753],[268,0.6964],[270,0.6522],[272,0.6656],[274,0.696],[276,0.6274],[278,0.6212],[282,0.5612],[284,0.6112],[286,0.607],[288,0.4798],[290,0.3986],[292,0.4182],[294,0.4573],[296,0.472],[298,0.5389],[302,0.6223],[304,0.6236],[306,0.592],[308,0.544],[310,0.4773],[312,0.4369],[314,0.3598],[316,0.2536],[318,0.2084],[320,0.1189],[322,0.0646],[324,-0.0821],[326,-0.1885],[328,-0.3727],[330,-0.4859],[332,-0.3879],[334,0.4432],[336,2.314],[338,4.7323],[340,7.2059],[342,9.2427],[344,10.3505],[346,10.8234],[348,11.2581],[350,11.6163],[352,11.8593],[354,11.9489],[356,7.0517],[360,-14.496],[362,-19.3932],[364,-9.9875],[366,-4.4649],[368,0]]
       }
       foot_r_fk_ctrl.Rotation.Y {
        type=MovieSceneScriptingFloatChannel
        default=0
        thinned_from=140
        keys [[0,0.102],[92,0.102],[94,0.285],[96,0.3993],[98,0.7545],[100,0.7308],[102,0.4982],[104,-0.3311],[106,-1.8274],[108,-3.1762],[110,-4.6828],[112,-5.6708],[114,-5.8446],[116,-5.709],[124,-5.3865],[134,-4.9209],[162,-3.494],[168,-3.21],[174,-2.9482],[178,-2.7892],[182,-2.645],[186,-2.5174],[190,-2.4085],[192,-2.3616],[194,-2.32],[196,-2.2841],[198,-2.255],[200,-2.2338],[202,-2.2209],[204,-2.2165],[206,-2.2796],[208,-2.4155],[210,-2.4359],[212,-2.4965],[214,-2.5074],[216,-2.6537],[218,-2.6883],[220,-2.5735],[222,-2.5998],[224,-2.9578],[226,-3.2103],[228,-3.3741],[230,-3.6869],[232,-3.9403],[234,-3.9905],[236,-3.8884],[238,-3.8834],[240,-3.9474],[242,-3.9865],[244,-4.0362],[246,-3.9743],[248,-3.9281],[250,-3.967],[252,-3.8835],[256,-3.7686],[258,-3.8269],[260,-3.8068],[262,-3.8615],[264,-4.0568],[266,-4.0495],[268,-3.9712],[270,-3.8319],[272,-3.9124],[274,-4.0691],[276,-4.1834],[278,-4.2129],[282,-4.358],[284,-4.4954],[286,-4.4973],[288,-4.4006],[290,-4.3368],[292,-4.3556],[296,-4.4127],[298,-4.4564],[300,-4.5278],[302,-4.6639],[304,-4.7375],[306,-4.5699],[308,-4.4965],[310,-4.3319],[312,-4.2753],[314,-4.4173],[316,-4.5959],[318,-4.6907],[320,-4.9332],[322,-5.0526],[324,-5.4152],[326,-5.6293],[328,-5.751],[332,-5.8914],[334,-6.3657],[336,-7.3782],[340,-9.9088],[342,-10.8725],[344,-11.2655],[354,-11.3158],[356,-5.8714],[360,18.0835],[362,23.5278],[364,8.5293],[366,3.1124],[368,0]]
       }
       foot_r_fk_ctrl.Rotation.Z {
        type=MovieSceneScriptingFloatChannel
        default=24.4589
        thinned_from=140
        keys [[0,36.4847],[94,36.4862],[96,36.5463],[98,36.7443],[100,36.7492],[102,37.2562],[104,38.5564],[106,40.2702],[108,41.8958],[110,43.1397],[112,43.9006],[114,43.8994],[116,43.1549],[128,40.5837],[166,30.8366],[176,28.5826],[182,27.4214],[188,26.4431],[194,25.6797],[198,25.3098],[202,25.085],[206,25.01],[208,24.8924],[210,24.8559],[214,24.9241],[216,24.8892],[220,25.0766],[222,25.1149],[224,25.0565],[228,24.779],[234,24.4818],[236,24.4855],[242,24.3115],[250,24.316],[252,24.4165],[254,24.4652],[256,24.5708],[262,24.6228],[266,24.6056],[270,24.6797],[276,24.5325],[286,24.4189],[290,24.5394],[294,24.5666],[304,24.4039],[310,24.77],[312,24.9392],[322,25.515],[326,25.8697],[330,25.9926],[332,26.2177],[354,25.8698],[356,24.8965],[360,20.7392],[362,19.7989],[364,22.7699],[366,23.8427],[368,24.4589]]
       }
       foot_r_fk_ctrl.Scale.X {type=MovieSceneScriptingFloatChannel, default=1, const [1,140]}
       foot_r_fk_ctrl.Scale.Y {type=MovieSceneScriptingFloatChannel, default=1, const [1,140]}
       foot_r_fk_ctrl.Scale.Z {type=MovieSceneScriptingFloatChannel, default=1, const [1,140]}
       ball_r_fk_ctrl.Location.X {
        type=MovieSceneScriptingFloatChannel
        default=0
        thinned_from=140
        keys [[0,0],[112,0.0001],[114,0.0002],[334,0],[338,0.0001],[340,-0.0001],[342,0.0001],[362,-0.0003],[364,0.0008],[366,0.003],[368,0]]
       }
       ball_r_fk_ctrl.Location.Y {
        type=MovieSceneScriptingFloatChannel
        default=-0.0001
        thinned_from=140
        keys [[0,0],[112,-0.0001],[114,-0.0003],[116,-0.0001],[118,-0.0003],[120,0],[334,0],[336,-0.0003],[340,-0.0004],[344,0],[356,-0.0002],[362,-0.0009],[364,-0.0044],[366,-0.0038],[368,-0.0001]]
       }
       ball_r_fk_ctrl.Location.Z {
        type=MovieSceneScriptingFloatChannel
        default=0
        thinned_from=140
        keys [[0,0],[112,0],[334,0],[340,0.0002],[362,0.0002],[364,-0.0014],[366,0.0016],[368,0]]
       }
       ball_r_fk_ctrl.Rotation.X {
        type=MovieSceneScriptingFloatChannel
        default=-0.0101
        thinned_from=140
        keys [[0,-0.0501],[94,-0.0501],[96,-0.0506],[98,-0.0504],[100,-0.0489],[102,-0.051],[104,-0.0514],[106,-0.0528],[108,-0.0534],[110,-0.0551],[112,-0.053],[114,-0.0535],[116,-0.0535],[120,-0.0532],[122,-0.0539],[124,-0.0541],[126,-0.0539],[128,-0.0548],[130,-0.0539],[132,-0.0559],[134,-0.0534],[136,-0.046],[138,-0.0425],[140,-0.0462],[142,-0.0516],[144,-0.0554],[148,-0.0576],[150,-0.0571],[152,-0.0559],[154,-0.0562],[156,-0.0569],[158,-0.0566],[160,-0.0549],[162,-0.0553],[164,-0.0553],[166,-0.0533],[168,-0.0536],[170,-0.0545],[172,-0.0541],[174,-0.0547],[176,-0.0549],[178,-0.0544],[180,-0.0535],[182,-0.0531],[184,-0.0518],[186,-0.052],[188,-0.0514],[190,-0.0515],[192,-0.0524],[194,-0.0524],[196,-0.0512],[198,-0.052],[200,-0.0512],[202,-0.0524],[204,-0.0543],[206,-0.0547],[208,-0.0522],[210,-0.052],[212,-0.0526],[214,-0.0538],[216,-0.0528],[218,-0.0534],[220,-0.053],[222,-0.0543],[224,-0.0536],[226,-0.0534],[228,-0.0543],[230,-0.055],[234,-0.0551],[236,-0.0547],[240,-0.0549],[244,-0.0546],[248,-0.0538],[250,-0.0538],[252,-0.0549],[256,-0.0562],[258,-0.056],[260,-0.056],[262,-0.0563],[264,-0.055],[266,-0.055],[268,-0.0563],[270,-0.0565],[272,-0.0566],[274,-0.0563],[276,-0.0554],[278,-0.0555],[280,-0.0559],[282,-0.0558],[284,-0.054],[286,-0.0537],[290,-0.0561],[294,-0.0562],[296,-0.056],[298,-0.0548],[300,-0.0546],[302,-0.0536],[304,-0.0529],[306,-0.0543],[308,-0.0545],[310,-0.0554],[312,-0.0567],[314,-0.0564],[318,-0.0566],[320,-0.0562],[322,-0.0567],[328,-0.0573],[330,-0.057],[332,-0.057],[334,-0.0577],[336,-0.0565],[338,-0.0564],[340,-0.0586],[342,-0.0561],[348,-0.0556],[354,-0.0543],[356,-0.0534],[360,-0.0505],[362,-0.0499],[364,-0.0365],[366,-0.0414],[368,-0.0101]]
       }
       ball_r_fk_ctrl.Rotation.Y {
        type=MovieSceneScriptingFloatChannel
        default=0.0103
        thinned_from=140
        keys [[0,-0.202],[94,-0.2022],[96,-0.2048],[98,-0.2036],[100,-0.1947],[102,-0.2076],[104,-0.2094],[106,-0.2179],[108,-0.222],[110,-0.2323],[112,-0.2194],[114,-0.2222],[116,-0.2227],[120,-0.2205],[122,-0.2247],[124,-0.2256],[126,-0.2245],[128,-0.2301],[130,-0.2248],[132,-0.2368],[134,-0.222],[136,-0.1778],[138,-0.1581],[140,-0.1792],[142,-0.2104],[144,-0.234],[146,-0.2412],[148,-0.2476],[150,-0.2444],[152,-0.2368],[154,-0.239],[156,-0.2432],[158,-0.241],[160,-0.231],[162,-0.2329],[164,-0.2334],[166,-0.221],[168,-0.2228],[170,-0.2283],[172,-0.2262],[174,-0.2293],[176,-0.2311],[178,-0.2278],[180,-0.2222],[182,-0.2201],[184,-0.2121],[186,-0.2131],[188,-0.2098],[190,-0.2102],[192,-0.2156],[194,-0.2156],[196,-0.2082],[198,-0.2135],[200,-0.2087],[202,-0.2157],[204,-0.227],[206,-0.2296],[208,-0.2147],[210,-0.2132],[212,-0.2168],[214,-0.2241],[216,-0.2181],[218,-0.2216],[220,-0.219],[222,-0.2271],[224,-0.2225],[226,-0.2217],[228,-0.2273],[230,-0.2313],[232,-0.2319],[234,-0.2318],[236,-0.2298],[240,-0.2307],[244,-0.2286],[246,-0.2259],[248,-0.2242],[250,-0.2242],[252,-0.2306],[256,-0.239],[258,-0.2373],[260,-0.2377],[262,-0.2396],[264,-0.2315],[266,-0.2313],[268,-0.2391],[270,-0.2409],[272,-0.2413],[274,-0.2392],[276,-0.2336],[278,-0.2347],[280,-0.2369],[282,-0.236],[284,-0.2252],[286,-0.2237],[290,-0.2378],[292,-0.2385],[294,-0.2385],[296,-0.2376],[298,-0.2299],[300,-0.2289],[302,-0.2226],[304,-0.2188],[306,-0.2274],[308,-0.2282],[310,-0.2341],[312,-0.2419],[314,-0.2399],[316,-0.2401],[318,-0.2414],[320,-0.2386],[322,-0.2419],[324,-0.243],[326,-0.2447],[328,-0.2455],[330,-0.2437],[332,-0.2439],[334,-0.2483],[336,-0.2409],[338,-0.2402],[340,-0.2536],[342,-0.2384],[344,-0.2377],[348,-0.2348],[352,-0.2304],[354,-0.2276],[356,-0.2218],[358,-0.2128],[360,-0.2046],[362,-0.201],[364,-0.1322],[366,-0.1618],[368,0.0103]]
       }
       ball_r_fk_ctrl.Rotation.Z {
        type=MovieSceneScriptingFloatChannel
        default=2.1589
        thinned_from=140
        keys [[0,0.005],[92,0.005],[94,0.003],[96,-0.0235],[98,-0.011],[100,0.0784],[102,-0.0513],[104,-0.0703],[106,-0.1561],[108,-0.1971],[110,-0.3005],[112,-0.1708],[114,-0.1993],[116,-0.2039],[118,-0.1922],[120,-0.1818],[122,-0.224],[124,-0.2333],[126,-0.2224],[128,-0.2793],[130,-0.225],[132,-0.3466],[134,-0.197],[136,0.2499],[138,0.4491],[140,0.2359],[142,-0.08],[144,-0.3185],[146,-0.3914],[148,-0.4556],[150,-0.4239],[152,-0.347],[154,-0.3687],[156,-0.4116],[158,-0.3894],[160,-0.2877],[162,-0.3078],[164,-0.3122],[166,-0.1872],[168,-0.2052],[170,-0.2609],[172,-0.239],[174,-0.2709],[176,-0.2888],[178,-0.2553],[180,-0.1996],[182,-0.1781],[184,-0.0969],[186,-0.1067],[188,-0.0735],[190,-0.0781],[192,-0.1322],[194,-0.1325],[196,-0.0576],[198,-0.111],[200,-0.0623],[202,-0.1338],[204,-0.2475],[206,-0.2743],[208,-0.1229],[210,-0.1082],[212,-0.1451],[214,-0.2182],[216,-0.1576],[218,-0.193],[220,-0.1667],[222,-0.2482],[224,-0.2025],[226,-0.1938],[228,-0.2502],[230,-0.2908],[232,-0.2969],[234,-0.2955],[236,-0.2754],[238,-0.28],[240,-0.2852],[242,-0.2737],[244,-0.2635],[246,-0.236],[248,-0.219],[250,-0.219],[252,-0.284],[254,-0.324],[256,-0.3685],[258,-0.3518],[260,-0.356],[262,-0.375],[264,-0.293],[266,-0.2906],[268,-0.37],[270,-0.3877],[272,-0.392],[274,-0.3704],[276,-0.3148],[278,-0.3254],[280,-0.3478],[282,-0.3387],[284,-0.2292],[286,-0.2144],[290,-0.357],[292,-0.3635],[294,-0.3641],[296,-0.3545],[298,-0.2767],[300,-0.2666],[302,-0.2036],[304,-0.1647],[306,-0.2519],[308,-0.2602],[310,-0.3194],[312,-0.3976],[314,-0.3776],[316,-0.3804],[318,-0.3935],[320,-0.3647],[322,-0.398],[324,-0.4092],[326,-0.4264],[328,-0.434],[330,-0.4161],[332,-0.4183],[334,-0.4622],[336,-0.388],[338,-0.3811],[340,-0.5162],[342,-0.3624],[344,-0.3555],[346,-0.3431],[348,-0.3269],[350,-0.3066],[352,-0.2824],[354,-0.254],[356,-0.195],[358,-0.1049],[360,-0.0224],[362,0.0138],[364,0.7121],[366,0.4145],[368,2.1589]]
       }
       ball_r_fk_ctrl.Scale.X {type=MovieSceneScriptingFloatChannel, default=1, const [1,140]}
       ball_r_fk_ctrl.Scale.Y {type=MovieSceneScriptingFloatChannel, default=1, const [1,140]}
       ball_r_fk_ctrl.Scale.Z {type=MovieSceneScriptingFloatChannel, default=1, const [1,140]}
       foot_r_ik_ctrl.Location.X {
        type=MovieSceneScriptingFloatChannel
        default=-12.4633
        thinned_from=140
        keys [[0,-45.2176],[92,-45.2176],[96,-45.6903],[98,-46.0821],[102,-47.0539],[104,-47.9861],[106,-48.6591],[108,-48.8897],[110,-49.2262],[112,-49.3502],[114,-49.0883],[116,-48.6922],[118,-48.6465],[120,-48.183],[122,-47.2677],[124,-46.7237],[126,-46.3616],[128,-46.4292],[132,-46.0047],[134,-45.3438],[136,-44.3974],[140,-43.3121],[142,-42.4795],[144,-41.8122],[150,-41.27],[152,-40.9266],[154,-40.421],[158,-39.0467],[162,-37.8368],[170,-35.9334],[172,-35.6056],[174,-35.1661],[178,-34.0166],[180,-33.5258],[182,-33.1357],[184,-32.8132],[186,-32.4229],[190,-31.8039],[196,-31.1411],[198,-30.9721],[200,-30.8651],[202,-30.863],[206,-30.9932],[216,-31.0052],[220,-30.8961],[224,-30.6225],[230,-30.0448],[244,-29.977],[250,-30.1186],[258,-30.0078],[262,-30.0783],[270,-30.0435],[300,-30.247],[306,-30.1432],[312,-29.8623],[318,-29.3334],[320,-29.0865],[322,-28.9545],[326,-28.1756],[328,-27.6958],[330,-27.2794],[332,-26.6805],[334,-26.3344],[336,-24.617],[338,-23.8599],[340,-23.2105],[342,-22.9449],[344,-21.8143],[348,-19.8863],[350,-19.6982],[352,-19.8865],[354,-20.4301],[356,-21.765],[358,-23.579],[360,-26.902],[362,-30.8586],[364,-34.7142],[366,-34.1813],[368,-12.4633]]
       }
       foot_r_ik_ctrl.Location.Y {
        type=MovieSceneScriptingFloatChannel
        default=72.0665
        thinned_from=140
        keys [[0,-1066.6144],[94,-1066.5095],[96,-992.0054],[98,-896.3278],[100,-787.7776],[102,-674.8952],[104,-565.4064],[106,-467.3192],[108,-389.8599],[110,-340.0027],[112,-308.9573],[114,-280.188],[116,-254.1094],[118,-232.3338],[120,-212.813],[122,-195.2792],[124,-180.3592],[126,-166.6459],[128,-154.384],[130,-144.0624],[132,-134.713],[134,-126.7874],[136,-119.7209],[140,-106.4766],[146,-87.5364],[148,-81.5602],[150,-75.9289],[152,-70.5055],[154,-65.3881],[156,-60.6143],[158,-56.1228],[160,-51.8445],[162,-47.8023],[164,-44.0633],[166,-40.5761],[168,-37.2491],[170,-34.0115],[172,-30.8587],[174,-28.0349],[176,-25.4076],[178,-23.0166],[180,-20.7629],[182,-18.6534],[184,-16.728],[186,-14.9675],[188,-13.438],[190,-12.0465],[192,-10.8999],[194,-10.0145],[196,-9.3595],[198,-8.8803],[200,-8.6041],[202,-8.4615],[204,-8.4297],[208,-8.4917],[212,-8.5866],[214,-8.5972],[216,-8.5686],[220,-8.4542],[222,-8.4646],[226,-8.5575],[228,-8.5786],[232,-8.6866],[234,-8.7083],[236,-8.6969],[238,-8.7059],[244,-8.7692],[246,-8.7329],[250,-8.6286],[252,-8.5494],[256,-8.4455],[260,-8.4499],[262,-8.4779],[264,-8.5357],[268,-8.6042],[270,-8.6537],[274,-8.8932],[276,-8.9862],[278,-9.0443],[284,-9.1844],[288,-9.1776],[290,-9.2059],[292,-9.2721],[294,-9.3677],[296,-9.4093],[298,-9.4837],[300,-9.5256],[302,-9.519],[304,-9.4366],[306,-9.2583],[308,-8.9987],[310,-8.8343],[312,-8.6967],[314,-8.6254],[316,-8.5953],[318,-8.5913],[322,-8.6526],[324,-8.7705],[326,-8.9253],[328,-9.1932],[330,-9.5005],[332,-9.6706],[334,-9.7455],[336,-10.934],[338,-9.8528],[340,-7.5565],[342,-4.5386],[344,-1.6436],[346,0.7255],[348,2.8502],[350,4.2362],[352,4.831],[354,5.1641],[356,4.4055],[358,2.1122],[360,-1.9439],[362,-7.9298],[364,-20.048],[366,-6.8553],[368,72.0665]]
       }
       foot_r_ik_ctrl.Location.Z {
        type=MovieSceneScriptingFloatChannel
        default=119.9044
        thinned_from=140
        keys [[0,89.7145],[92,89.7145],[102,88.0563],[104,87.957],[106,88.9192],[108,88.9113],[110,89.9341],[114,88.5277],[116,87.0321],[120,77.4489],[122,73.4048],[126,68.3799],[128,65.6816],[130,63.5642],[132,61.2747],[134,59.1627],[140,52.027],[142,49.7796],[146,45.0122],[148,42.9182],[152,39.1651],[154,37.4174],[156,35.7879],[158,34.2575],[160,32.8019],[162,31.4146],[164,30.1176],[172,25.2976],[174,24.2372],[176,23.2529],[178,22.3448],[180,21.4847],[182,20.6815],[184,19.9514],[186,19.3367],[188,18.811],[190,18.377],[192,18.0554],[194,17.8607],[196,17.7932],[206,17.8545],[210,17.994],[216,18.3878],[220,18.7606],[224,19.1991],[230,19.6529],[232,19.9529],[244,21.8742],[252,22.9844],[258,23.6712],[264,24.2137],[272,24.6357],[274,24.6872],[296,24.0038],[302,23.6809],[308,23.2534],[314,22.9791],[318,22.6799],[320,22.4752],[322,22.3317],[326,21.8612],[328,21.6623],[330,21.5246],[332,21.3178],[334,21.3505],[336,21.2633],[338,20.2784],[340,19.1952],[342,18.5956],[344,19.3528],[346,19.9487],[348,21.1618],[350,22.4955],[354,25.5698],[356,26.6263],[358,27.0937],[360,25.6474],[362,23.1551],[364,23.5551],[366,39.8854],[368,119.9044]]
       }
       foot_r_ik_ctrl.Rotation.X {
        type=MovieSceneScriptingFloatChannel
        default=79.5857
        thinned_from=140
        keys [[0,127.4674],[96,127.5938],[100,127.7084],[102,128.7956],[104,132.008],[108,142.4219],[110,148.2843],[112,152.9393],[114,156.5987],[116,158.3264],[118,154.4547],[120,150.1629],[122,146.3225],[128,137.9655],[132,133.037],[138,126.2107],[142,121.1344],[146,115.0211],[154,103.9193],[160,96.3222],[164,91.7247],[172,83.3194],[178,77.7261],[182,74.5146],[186,71.8138],[190,69.6931],[192,68.9032],[194,68.3355],[198,67.7371],[202,67.6165],[206,67.7558],[210,67.6881],[214,67.7717],[222,67.6502],[224,67.494],[230,66.5868],[254,66.3373],[274,66.6129],[286,66.3214],[294,66.6233],[308,66.4634],[314,66.6493],[318,66.5545],[322,66.3767],[332,65.2006],[334,65.0472],[336,64.6952],[338,63.8975],[340,63.2571],[342,62.8345],[344,62.2664],[348,61.9994],[350,62.7682],[354,65.5626],[358,67.6609],[360,71.4924],[362,78.0575],[364,91.2018],[366,77.7048],[368,79.5857]]
       }
       foot_r_ik_ctrl.Rotation.Y {
        type=MovieSceneScriptingFloatChannel
        default=-3.2551
        thinned_from=140
        keys [[0,3.6885],[92,3.6885],[94,3.6112],[96,3.3368],[98,2.8588],[100,2.152],[102,1.9288],[104,2.053],[106,1.9879],[108,2.0801],[110,1.8982],[112,2.7268],[114,2.9522],[116,3.4398],[118,3.6289],[120,3.9619],[122,4.475],[124,5.0282],[126,5.3676],[128,5.5771],[130,5.9312],[134,6.3012],[136,6.3347],[140,6.1769],[142,6.3904],[144,6.6483],[146,6.6202],[150,6.4454],[152,6.3041],[156,5.8849],[158,5.6494],[160,5.3474],[162,4.9964],[164,4.5936],[166,4.2075],[168,3.8551],[170,3.416],[172,2.8021],[174,2.3174],[176,1.9778],[178,1.6685],[180,1.3144],[182,0.9107],[184,0.513],[186,0.2023],[188,-0.0798],[190,-0.3513],[192,-0.5812],[194,-0.7409],[196,-0.8347],[198,-0.8922],[200,-0.906],[202,-0.9314],[204,-0.9594],[206,-0.9563],[208,-0.9799],[210,-0.9714],[212,-0.8648],[214,-0.8109],[216,-0.894],[218,-0.9487],[220,-0.9459],[222,-0.9002],[224,-0.9449],[226,-1.1509],[228,-1.3785],[230,-1.7749],[232,-1.8633],[234,-1.8444],[236,-1.8445],[238,-1.9787],[240,-2.0303],[242,-2.0005],[244,-1.9762],[246,-2.0401],[248,-2.0926],[250,-2.1326],[252,-2.111],[254,-2.1024],[256,-2.1215],[258,-2.1933],[260,-2.2825],[262,-2.3086],[264,-2.4061],[266,-2.4507],[268,-2.5666],[274,-2.5095],[276,-2.5751],[278,-2.564],[280,-2.591],[282,-2.6085],[284,-2.5932],[286,-2.6404],[288,-2.7439],[290,-2.8071],[292,-2.752],[294,-2.678],[296,-2.6689],[298,-2.6076],[300,-2.593],[302,-2.6034],[304,-2.6607],[306,-2.6934],[308,-2.7614],[310,-2.77],[312,-2.7327],[314,-2.7493],[316,-2.7456],[318,-2.6536],[322,-2.6078],[324,-2.5799],[326,-2.4234],[328,-2.238],[330,-2.0183],[332,-1.6456],[334,-1.1477],[336,0.9601],[338,1.5994],[340,1.7212],[342,1.2565],[344,1.2064],[348,0.5645],[350,-0.0089],[352,-0.3624],[354,-0.6457],[356,-3.6399],[358,-8.9375],[360,-14.4997],[362,-17.2667],[364,-12.5497],[366,-8.1422],[368,-3.2551]]
       }
       foot_r_ik_ctrl.Rotation.Z {
        type=MovieSceneScriptingFloatChannel
        default=-14.0372
        thinned_from=140
        keys [[0,-14.7821],[92,-14.7821],[94,-14.8889],[96,-15.0906],[98,-15.021],[100,-15.5678],[102,-15.9872],[106,-17.669],[108,-18.2126],[110,-18.908],[112,-18.4213],[114,-17.335],[116,-16.8146],[118,-17.4234],[120,-17.8732],[122,-18.0989],[124,-18.4952],[126,-19.141],[128,-20.2025],[130,-21.0468],[132,-21.9904],[134,-22.5134],[136,-22.7098],[138,-23.221],[142,-24.5152],[144,-25.5974],[150,-30.2579],[152,-31.7385],[154,-32.9435],[156,-33.8524],[160,-35.5656],[162,-36.3408],[166,-37.7032],[170,-39.2055],[172,-39.8301],[178,-41.1107],[186,-42.248],[190,-42.6049],[192,-42.6955],[194,-42.6999],[198,-42.4188],[200,-42.1076],[206,-41.6903],[214,-41.7183],[218,-41.8054],[220,-41.7626],[222,-41.944],[224,-42.4559],[226,-42.8376],[228,-43.1244],[232,-44.0753],[234,-44.4276],[236,-44.5519],[244,-44.6644],[248,-44.5085],[256,-44.3791],[258,-44.4067],[262,-44.262],[264,-44.2884],[266,-44.2032],[270,-43.6207],[272,-43.504],[278,-43.7986],[284,-44.1948],[286,-44.0575],[290,-43.5021],[294,-43.2852],[300,-43.2213],[304,-43.4118],[306,-43.3994],[308,-43.6377],[310,-43.7442],[312,-43.9845],[316,-44.7448],[318,-45.0342],[320,-45.4191],[322,-45.5841],[324,-45.9439],[326,-46.1204],[328,-46.0369],[330,-45.7882],[332,-45.3793],[334,-45.2547],[336,-44.9172],[338,-45.1843],[340,-45.1082],[342,-45.7268],[346,-46.432],[348,-47.1634],[350,-47.7214],[352,-47.6739],[354,-47.1851],[356,-41.5005],[358,-28.9992],[360,-17.4971],[362,-12.8121],[364,-21.3293],[366,-15.8116],[368,-14.0372]]
       }
       foot_r_ik_ctrl.Scale.X {type=MovieSceneScriptingFloatChannel, default=1, const [1,140]}
       foot_r_ik_ctrl.Scale.Y {type=MovieSceneScriptingFloatChannel, default=1, const [1,140]}
       foot_r_ik_ctrl.Scale.Z {type=MovieSceneScriptingFloatChannel, default=1, const [1,140]}
       ball_r_ik_ctrl.Location.X {
        type=MovieSceneScriptingFloatChannel
        default=0
        thinned_from=140
        keys [[0,0],[334,0],[336,-0.0002],[362,-0.0003],[364,0.0016],[366,-0.0033],[368,0]]
       }
       ball_r_ik_ctrl.Location.Y {
        type=MovieSceneScriptingFloatChannel
        default=-0.0001
        thinned_from=140
        keys [[0,0],[112,-0.0001],[114,-0.0003],[116,-0.0001],[118,-0.0003],[120,-0.0001],[334,0],[336,-0.0002],[340,-0.0003],[344,0],[356,-0.0002],[362,-0.0009],[364,-0.0077],[366,-0.0061],[368,-0.0001]]
       }
       ball_r_ik_ctrl.Location.Z {
        type=MovieSceneScriptingFloatChannel
        default=0
        thinned_from=140
        keys [[0,0],[112,0.0001],[114,0.0002],[334,0],[338,0.0001],[340,-0.0001],[342,0.0001],[362,-0.0003],[364,0.0014],[366,0.0052],[368,0]]
       }
       ball_r_ik_ctrl.Rotation.X {
        type=MovieSceneScriptingFloatChannel
        default=2.15
        thinned_from=140
        keys [[0,-0.0146],[92,-0.0146],[94,-0.0165],[96,-0.0431],[98,-0.0306],[100,0.0593],[102,-0.0711],[104,-0.0902],[106,-0.1764],[108,-0.2176],[110,-0.3215],[112,-0.1912],[114,-0.2198],[116,-0.2244],[118,-0.2127],[120,-0.2022],[122,-0.2446],[124,-0.254],[126,-0.243],[128,-0.3002],[130,-0.2456],[132,-0.3678],[134,-0.2175],[136,0.2316],[138,0.4318],[140,0.2175],[142,-0.0999],[144,-0.3396],[146,-0.4128],[148,-0.4774],[150,-0.4455],[152,-0.3683],[154,-0.39],[156,-0.4331],[158,-0.4108],[160,-0.3086],[162,-0.3288],[164,-0.3333],[166,-0.2077],[168,-0.2258],[170,-0.2817],[172,-0.2597],[174,-0.2918],[176,-0.3098],[178,-0.2761],[180,-0.2201],[182,-0.1985],[184,-0.1169],[186,-0.1268],[188,-0.0934],[190,-0.098],[192,-0.1524],[194,-0.1527],[196,-0.0775],[198,-0.1311],[200,-0.0821],[202,-0.154],[204,-0.2683],[206,-0.2952],[208,-0.143],[210,-0.1283],[212,-0.1653],[214,-0.2388],[216,-0.1779],[218,-0.2135],[220,-0.1871],[222,-0.269],[224,-0.2231],[226,-0.2143],[228,-0.271],[230,-0.3118],[232,-0.3179],[234,-0.3165],[236,-0.2963],[238,-0.3009],[240,-0.3062],[242,-0.2946],[244,-0.2843],[246,-0.2567],[248,-0.2396],[250,-0.2396],[252,-0.305],[254,-0.3451],[256,-0.3899],[258,-0.3731],[260,-0.3773],[262,-0.3963],[264,-0.3139],[266,-0.3116],[268,-0.3914],[270,-0.4091],[272,-0.4134],[274,-0.3918],[276,-0.3359],[278,-0.3465],[280,-0.3691],[282,-0.3599],[284,-0.2499],[286,-0.235],[290,-0.3783],[292,-0.3848],[294,-0.3854],[296,-0.3757],[298,-0.2976],[300,-0.2875],[302,-0.2241],[304,-0.185],[306,-0.2727],[308,-0.281],[310,-0.3405],[312,-0.4191],[314,-0.399],[316,-0.4018],[318,-0.415],[320,-0.3861],[322,-0.4195],[324,-0.4308],[326,-0.448],[328,-0.4557],[330,-0.4377],[332,-0.4399],[334,-0.4841],[336,-0.4095],[338,-0.4026],[340,-0.5383],[342,-0.3838],[344,-0.3768],[346,-0.3643],[348,-0.348],[350,-0.3277],[352,-0.3033],[354,-0.2748],[356,-0.2155],[358,-0.1249],[360,-0.042],[362,-0.0057],[364,-0.3237],[366,-0.8327],[368,2.15]]
       }
       ball_r_ik_ctrl.Rotation.Y {
        type=MovieSceneScriptingFloatChannel
        default=-0.1961
        thinned_from=140
        keys [[0,-0.2019],[98,-0.202],[100,-0.2018],[110,-0.2027],[130,-0.2025],[132,-0.2027],[134,-0.2024],[136,-0.2014],[138,-0.2008],[144,-0.2027],[158,-0.2028],[202,-0.2023],[206,-0.2026],[282,-0.2027],[304,-0.2023],[334,-0.203],[338,-0.2028],[340,-0.2032],[362,-0.2018],[364,-0.2023],[366,-0.2037],[368,-0.1961]]
       }
       ball_r_ik_ctrl.Rotation.Z {
        type=MovieSceneScriptingFloatChannel
        default=0.0066
        thinned_from=140
        keys [[0,0.0485],[94,0.0485],[96,0.049],[98,0.0488],[100,0.047],[102,0.0496],[104,0.0499],[106,0.0516],[108,0.0524],[110,0.0544],[112,0.0519],[114,0.0524],[116,0.0525],[120,0.0521],[122,0.0529],[124,0.0531],[126,0.0529],[128,0.054],[130,0.0529],[132,0.0553],[134,0.0524],[136,0.0437],[138,0.0398],[140,0.044],[142,0.0501],[144,0.0547],[148,0.0574],[150,0.0568],[152,0.0553],[154,0.0557],[156,0.0566],[158,0.0562],[160,0.0542],[162,0.0546],[164,0.0546],[166,0.0522],[168,0.0526],[170,0.0536],[172,0.0532],[174,0.0538],[176,0.0542],[178,0.0535],[180,0.0524],[182,0.052],[184,0.0504],[186,0.0506],[188,0.05],[190,0.0501],[192,0.0511],[194,0.0511],[196,0.0497],[198,0.0507],[200,0.0498],[202,0.0512],[204,0.0534],[206,0.0539],[208,0.0509],[210,0.0507],[212,0.0514],[214,0.0528],[216,0.0516],[218,0.0523],[220,0.0518],[222,0.0534],[224,0.0525],[226,0.0523],[228,0.0534],[230,0.0542],[234,0.0543],[236,0.0539],[240,0.0541],[244,0.0537],[246,0.0532],[248,0.0528],[250,0.0528],[252,0.0541],[256,0.0557],[258,0.0554],[260,0.0555],[262,0.0559],[264,0.0543],[266,0.0542],[268,0.0558],[270,0.0561],[272,0.0562],[274,0.0558],[276,0.0547],[278,0.0549],[280,0.0553],[282,0.0552],[284,0.053],[286,0.0527],[290,0.0555],[294,0.0557],[296,0.0555],[298,0.0539],[300,0.0537],[302,0.0525],[304,0.0518],[306,0.0534],[308,0.0536],[310,0.0548],[312,0.0563],[314,0.0559],[318,0.0562],[320,0.0556],[322,0.0563],[328,0.057],[330,0.0567],[332,0.0567],[334,0.0576],[336,0.0561],[338,0.056],[340,0.0586],[342,0.0556],[346,0.0553],[352,0.054],[354,0.0535],[356,0.0523],[360,0.049],[362,0.0483],[364,0.0544],[366,0.0645],[368,0.0066]]
       }
       ball_r_ik_ctrl.Scale.X {type=MovieSceneScriptingFloatChannel, default=1, const [1,140]}
       ball_r_ik_ctrl.Scale.Y {type=MovieSceneScriptingFloatChannel, default=1, const [1,140]}
       ball_r_ik_ctrl.Scale.Z {type=MovieSceneScriptingFloatChannel, default=1, const [1,140]}
       foot_bk1_r_ctrl.Location.X {type=MovieSceneScriptingFloatChannel, default=0, const [0,140]}
       foot_bk1_r_ctrl.Location.Y {type=MovieSceneScriptingFloatChannel, default=0, const [0,140]}
       foot_bk1_r_ctrl.Location.Z {type=MovieSceneScriptingFloatChannel, default=0, const [0,140]}
       foot_bk1_r_ctrl.Rotation.X {type=MovieSceneScriptingFloatChannel, default=0, const [0,140]}
       foot_bk1_r_ctrl.Rotation.Y {type=MovieSceneScriptingFloatChannel, default=0, const [0,140]}
       foot_bk1_r_ctrl.Rotation.Z {type=MovieSceneScriptingFloatChannel, default=0, const [0,140]}
       foot_bk1_r_ctrl.Scale.X {type=MovieSceneScriptingFloatChannel, default=1, const [1,140]}
       foot_bk1_r_ctrl.Scale.Y {type=MovieSceneScriptingFloatChannel, default=1, const [1,140]}
       foot_bk1_r_ctrl.Scale.Z {type=MovieSceneScriptingFloatChannel, default=1, const [1,140]}
       foot_roll_r_ctrl.Location.X {type=MovieSceneScriptingFloatChannel, default=0, const [0,140]}
       foot_roll_r_ctrl.Location.Y {type=MovieSceneScriptingFloatChannel, default=0, const [0,140]}
       foot_roll_r_ctrl.Location.Z {type=MovieSceneScriptingFloatChannel, default=0, const [0,140]}
       foot_roll_r_ctrl.Rotation.X {type=MovieSceneScriptingFloatChannel, default=0, const [0,140]}
       foot_roll_r_ctrl.Rotation.Y {type=MovieSceneScriptingFloatChannel, default=0, const [0,140]}
       foot_roll_r_ctrl.Rotation.Z {type=MovieSceneScriptingFloatChannel, default=0, const [0,140]}
       foot_roll_r_ctrl.Scale.X {type=MovieSceneScriptingFloatChannel, default=1, const [1,140]}
       foot_roll_r_ctrl.Scale.Y {type=MovieSceneScriptingFloatChannel, default=1, const [1,140]}
       foot_roll_r_ctrl.Scale.Z {type=MovieSceneScriptingFloatChannel, default=1, const [1,140]}
       roll_blend_r_ctrl {type=MovieSceneScriptingFloatChannel, default=1, const [1,140]}
       tip_r_ctrl.Location.X {type=MovieSceneScriptingFloatChannel, default=0, const [0,140]}
       tip_r_ctrl.Location.Y {type=MovieSceneScriptingFloatChannel, default=0, const [0,140]}
       tip_r_ctrl.Location.Z {type=MovieSceneScriptingFloatChannel, default=0, const [0,140]}
       tip_r_ctrl.Rotation.X {type=MovieSceneScriptingFloatChannel, default=0, const [0,140]}
       tip_r_ctrl.Rotation.Y {type=MovieSceneScriptingFloatChannel, default=0, const [0,140]}
       tip_r_ctrl.Rotation.Z {type=MovieSceneScriptingFloatChannel, default=0, const [0,140]}
       tip_r_ctrl.Scale.X {type=MovieSceneScriptingFloatChannel, default=1, const [1,140]}
       tip_r_ctrl.Scale.Y {type=MovieSceneScriptingFloatChannel, default=1, const [1,140]}
       tip_r_ctrl.Scale.Z {type=MovieSceneScriptingFloatChannel, default=1, const [1,140]}
       heel_r_ctrl.Location.X {type=MovieSceneScriptingFloatChannel, default=0, const [0,140]}
       heel_r_ctrl.Location.Y {type=MovieSceneScriptingFloatChannel, default=0, const [0,140]}
       heel_r_ctrl.Location.Z {type=MovieSceneScriptingFloatChannel, default=0, const [0,140]}
       heel_r_ctrl.Rotation.X {type=MovieSceneScriptingFloatChannel, default=0, const [0,140]}
       heel_r_ctrl.Rotation.Y {type=MovieSceneScriptingFloatChannel, default=0, const [0,140]}
       heel_r_ctrl.Rotation.Z {type=MovieSceneScriptingFloatChannel, default=0, const [0,140]}
       heel_r_ctrl.Scale.X {type=MovieSceneScriptingFloatChannel, default=1, const [1,140]}
       heel_r_ctrl.Scale.Y {type=MovieSceneScriptingFloatChannel, default=1, const [1,140]}
       heel_r_ctrl.Scale.Z {type=MovieSceneScriptingFloatChannel, default=1, const [1,140]}
       leg_r_softness {type=MovieSceneScriptingFloatChannel, default=0, const [0,140]}
       leg_r_pv_ik_ctrl.Location.X {
        type=MovieSceneScriptingFloatChannel
        default=-0.6331
        thinned_from=140
        keys [[0,-34.1276],[94,-34.1258],[96,-34.3153],[98,-34.7566],[100,-35.4413],[104,-35.9697],[106,-36.5665],[108,-36.2685],[110,-36.4481],[112,-35.3112],[114,-33.8775],[116,-31.7795],[118,-29.7943],[120,-27.4612],[122,-24.9931],[124,-23.0009],[126,-21.1324],[130,-17.5048],[132,-15.7536],[134,-14.1249],[136,-12.7472],[138,-11.4464],[140,-10.0353],[142,-8.3801],[144,-6.4202],[146,-4.4426],[150,-0.7754],[152,1.022],[154,2.5824],[156,3.9177],[158,5.1332],[160,6.3087],[162,7.3559],[164,8.2208],[166,9.0659],[168,9.9731],[170,10.8244],[172,11.5013],[174,12.0516],[176,12.6815],[178,13.259],[180,13.7113],[182,14.1103],[186,14.8177],[188,15.1126],[190,15.3616],[192,15.5473],[194,15.676],[196,15.7155],[198,15.6997],[200,15.5156],[204,15.2574],[206,15.0793],[210,14.9176],[214,14.8277],[216,14.7589],[218,14.7441],[220,14.8136],[224,15.1167],[228,15.5078],[230,15.7998],[232,15.9933],[234,16.2268],[236,16.4091],[238,16.481],[246,16.4066],[250,16.333],[256,16.4011],[258,16.3928],[260,16.3386],[264,16.1305],[266,16.0839],[268,15.9534],[270,15.788],[272,15.5826],[274,15.4599],[276,15.4277],[280,15.5234],[282,15.5477],[284,15.5073],[288,15.2788],[294,14.8551],[296,14.8044],[300,14.6375],[302,14.6092],[304,14.6641],[306,14.8321],[308,15.179],[312,15.7998],[316,16.3517],[320,16.7952],[322,16.8807],[324,17.0196],[326,17.112],[328,17.1052],[330,16.9862],[332,16.7406],[334,16.2788],[336,15.5438],[338,14.6218],[340,13.391],[342,12.9567],[344,12.9325],[346,13.3137],[350,14.447],[352,14.6555],[354,14.3076],[356,14.0363],[358,12.9947],[360,11.4303],[362,8.5089],[364,4.2799],[366,-9.6759],[368,-0.6331]]
       }
       leg_r_pv_ik_ctrl.Location.Y {
        type=MovieSceneScriptingFloatChannel
        default=84.62
        thinned_from=140
        keys [[0,-1073.0986],[94,-1072.9324],[96,-998.2912],[98,-902.4139],[100,-793.7465],[102,-680.3301],[104,-569.785],[106,-470.5991],[108,-391.3336],[110,-340.0121],[112,-306.7742],[114,-275.7492],[116,-247.0939],[118,-221.1242],[120,-197.5135],[122,-176.5743],[124,-159.452],[126,-143.708],[128,-129.6859],[130,-118.0458],[132,-107.5579],[134,-98.6017],[136,-90.4408],[138,-82.9523],[140,-75.8142],[142,-69.1552],[144,-62.7419],[146,-56.8233],[148,-51.5208],[150,-46.7627],[152,-42.3472],[154,-38.2054],[156,-34.3184],[158,-30.7597],[160,-27.5398],[162,-24.5646],[164,-21.8594],[166,-19.4639],[168,-17.3368],[170,-15.3686],[172,-13.4913],[174,-11.7493],[176,-10.2173],[178,-8.8861],[180,-7.6241],[182,-6.5002],[184,-5.5179],[186,-4.6198],[188,-3.8531],[190,-3.1545],[192,-2.553],[194,-2.0208],[196,-1.5127],[198,-1.0688],[200,-0.6186],[202,-0.3524],[204,-0.2035],[206,-0.0446],[208,0.0439],[210,0.1036],[212,0.1438],[214,0.1682],[216,0.2163],[218,0.2255],[220,0.1714],[222,0.0464],[224,-0.1022],[226,-0.2782],[228,-0.4504],[230,-0.7113],[232,-0.8783],[234,-1.0691],[236,-1.2139],[238,-1.2593],[242,-1.2269],[244,-1.2006],[250,-1.1012],[252,-1.1007],[254,-1.089],[256,-1.0694],[258,-1.0547],[260,-1.0225],[262,-0.956],[264,-0.8755],[266,-0.8483],[268,-0.7799],[270,-0.7005],[272,-0.6196],[274,-0.605],[276,-0.6855],[278,-0.807],[280,-0.9411],[282,-1.0248],[284,-1.0164],[286,-0.932],[288,-0.836],[290,-0.7129],[292,-0.6221],[294,-0.5446],[296,-0.5204],[298,-0.474],[300,-0.4241],[302,-0.3821],[304,-0.3786],[306,-0.423],[310,-0.6356],[312,-0.7592],[314,-0.8643],[316,-1.0057],[318,-1.1317],[320,-1.2708],[322,-1.3343],[324,-1.4511],[326,-1.5533],[328,-1.6056],[330,-1.5855],[332,-1.4126],[334,-1.0903],[336,-0.8565],[338,-0.1767],[340,1.0445],[342,1.5823],[344,1.7722],[346,1.1321],[348,-0.1729],[350,-1.7325],[352,-2.8672],[354,-3.5014],[356,-3.5046],[358,-1.9744],[360,0.2559],[362,2.5291],[364,6.4041],[366,20.3988],[368,84.62]]
       }
       leg_r_pv_ik_ctrl.Location.Z {
        type=MovieSceneScriptingFloatChannel
        default=67.326
        thinned_from=140
        keys [[0,5.8528],[92,5.8528],[94,5.5498],[98,4.7388],[100,4.5398],[102,3.7541],[104,2.1673],[106,0.6407],[108,-1.8894],[110,-3.8129],[112,-6.6074],[114,-8.851],[116,-10.5798],[118,-11.4039],[120,-11.6763],[122,-11.2926],[124,-10.5152],[126,-9.4599],[128,-8.3215],[130,-7.0994],[132,-5.9283],[134,-4.8914],[136,-3.9961],[138,-3.0881],[140,-1.9825],[142,-0.642],[144,1.0455],[146,2.8037],[148,4.4281],[150,5.9923],[152,7.5854],[154,9.0911],[156,10.513],[158,11.8322],[160,13.0774],[162,14.2399],[164,15.2111],[170,17.8122],[172,18.5769],[174,19.2649],[176,20.0239],[178,20.7136],[180,21.3293],[182,21.8694],[186,22.8613],[190,23.7548],[192,24.144],[196,24.7964],[198,25.0619],[200,25.198],[202,25.2271],[206,25.122],[210,25.1088],[212,25.1576],[214,25.2843],[216,25.4944],[226,27.0238],[230,27.4967],[250,30.4237],[254,31.093],[256,31.3743],[268,32.2339],[272,32.3641],[288,31.9164],[296,31.423],[302,30.9731],[306,30.879],[312,30.9929],[316,30.8931],[330,30.1196],[332,30.1615],[334,30.5101],[336,30.9816],[338,32.0448],[340,33.8803],[342,36.5501],[344,40.5725],[346,43.8774],[348,47.493],[350,50.1589],[352,52.0657],[354,53.6017],[356,53.3927],[358,51.0659],[360,44.3251],[362,33.8922],[364,10.1146],[366,9.4608],[368,67.326]]
       }
       leg_r_pv_ik_ctrl.Rotation.X {
        type=MovieSceneScriptingFloatChannel
        default=55.1633
        thinned_from=140
        keys [[0,90.7542],[92,90.7542],[98,90.1702],[100,90.126],[102,89.2855],[106,85.3769],[110,78.5568],[112,73.5602],[114,68.2393],[116,62.4499],[118,57.1901],[120,51.7834],[122,46.8143],[126,39.158],[128,35.2068],[132,28.2986],[136,22.0359],[142,12.771],[144,9.4321],[146,6.2032],[148,3.3245],[150,0.6865],[152,-1.8748],[154,-4.2561],[156,-6.458],[158,-8.486],[160,-10.3829],[162,-12.1532],[164,-13.7068],[168,-16.5447],[170,-17.934],[172,-19.2494],[176,-21.6326],[178,-22.7423],[180,-23.7826],[182,-24.7478],[186,-26.499],[190,-28.0092],[192,-28.6496],[194,-29.1839],[196,-29.6439],[198,-30.0236],[200,-30.2443],[202,-30.3402],[208,-30.216],[212,-30.019],[216,-30.0523],[222,-30.3427],[226,-30.3881],[248,-30.0729],[256,-30.3134],[266,-30.1272],[270,-30.1378],[278,-29.8974],[292,-29.8073],[300,-29.5581],[304,-29.604],[306,-29.7113],[308,-29.9002],[312,-30.1383],[314,-30.1726],[318,-30.089],[330,-29.4938],[332,-29.5163],[334,-29.9292],[336,-30.7828],[338,-33.3901],[340,-37.5439],[342,-42.6832],[344,-48.2532],[348,-59.0671],[350,-63.5507],[352,-66.8387],[354,-69.2733],[356,-69.4249],[358,-66.5964],[360,-60.0213],[362,-50.6917],[364,-22.2033],[366,15.1751],[368,55.1633]]
       }
       leg_r_pv_ik_ctrl.Rotation.Y {
        type=MovieSceneScriptingFloatChannel
        default=-5.338
        thinned_from=140
        keys [[0,0.4582],[92,0.4582],[94,0.5507],[96,0.3667],[98,-0.0811],[100,-0.7793],[102,-1.0664],[104,-1.3858],[106,-2.4168],[108,-2.8756],[110,-3.9043],[112,-4.0041],[114,-4.2273],[116,-4.1389],[118,-4.4022],[120,-4.4979],[122,-4.3271],[124,-3.9923],[126,-3.8453],[128,-3.8025],[130,-3.5594],[132,-3.4567],[134,-3.319],[136,-3.3078],[138,-3.3887],[140,-3.4364],[142,-3.1705],[144,-2.8338],[146,-2.7551],[148,-2.717],[150,-2.6593],[152,-2.642],[154,-2.6856],[156,-2.7195],[158,-2.7792],[160,-2.9045],[162,-3.0803],[164,-3.3166],[166,-3.5427],[168,-3.7367],[170,-4.0213],[172,-4.4885],[174,-4.8383],[176,-5.0471],[178,-5.2366],[180,-5.4814],[184,-6.0984],[186,-6.3315],[190,-6.7637],[192,-6.9504],[194,-7.0803],[196,-7.1559],[198,-7.2027],[200,-7.2139],[204,-7.2786],[206,-7.275],[212,-7.1513],[216,-7.1046],[218,-7.0671],[220,-6.9983],[222,-6.8291],[224,-6.6073],[226,-6.332],[228,-6.0931],[230,-5.759],[232,-5.6245],[234,-5.5239],[236,-5.4593],[238,-5.3829],[240,-5.3192],[242,-5.2995],[244,-5.2913],[248,-5.4181],[250,-5.4481],[252,-5.4603],[254,-5.4422],[256,-5.4114],[258,-5.4043],[260,-5.4362],[262,-5.4536],[268,-5.3887],[270,-5.3763],[272,-5.3431],[278,-5.2038],[282,-5.1276],[284,-5.1184],[286,-5.1566],[288,-5.1831],[290,-5.1924],[294,-5.1215],[298,-5.0949],[302,-5.0977],[304,-5.1253],[308,-5.2481],[310,-5.2712],[312,-5.2432],[314,-5.1525],[316,-4.9912],[318,-4.8186],[320,-4.6054],[322,-4.4919],[324,-4.1786],[326,-3.8235],[328,-3.3725],[330,-2.9822],[332,-2.6355],[334,-2.7302],[336,-2.001],[338,-3.0853],[340,-4.7393],[342,-6.668],[344,-7.4728],[346,-8.1868],[348,-8.8684],[350,-9.9768],[352,-10.9807],[354,-11.7852],[356,-11.7018],[358,-9.6146],[360,-7.1254],[362,-4.7958],[364,-0.715],[366,-5.8553],[368,-5.338]]
       }
       leg_r_pv_ik_ctrl.Rotation.Z {
        type=MovieSceneScriptingFloatChannel
        default=-13.1412
        thinned_from=140
        keys [[0,-16.0894],[92,-16.0894],[94,-16.2906],[96,-16.5602],[98,-16.9413],[100,-17.4423],[102,-17.8647],[104,-18.7123],[106,-19.6154],[108,-20.13],[110,-21.0331],[112,-21.6917],[114,-22.2561],[116,-22.7428],[118,-22.8492],[120,-22.7114],[122,-22.3762],[124,-22.3638],[126,-22.5736],[128,-23.1642],[130,-23.6042],[132,-24.1323],[134,-24.2831],[136,-24.0983],[138,-24.23],[142,-24.7326],[144,-25.366],[148,-27.6755],[152,-30.0444],[154,-30.9639],[160,-32.921],[162,-33.5279],[166,-34.6507],[170,-35.9573],[172,-36.5026],[178,-37.6172],[186,-38.6457],[190,-38.9753],[192,-39.0586],[194,-39.0622],[198,-38.7847],[200,-38.4813],[206,-38.0252],[216,-37.7621],[218,-37.7647],[220,-37.8568],[228,-38.7268],[230,-39.1135],[236,-39.9558],[238,-40.0547],[250,-39.9494],[258,-40.0072],[262,-39.8548],[264,-39.7146],[266,-39.6592],[272,-39.1656],[274,-39.0689],[282,-39.3839],[284,-39.3305],[294,-38.5839],[302,-38.2997],[304,-38.3273],[306,-38.4755],[312,-39.3387],[320,-40.2235],[326,-40.3543],[328,-40.2283],[330,-39.968],[332,-39.4535],[334,-38.6175],[338,-34.7866],[340,-32.2777],[342,-30.7191],[344,-29.6431],[346,-29.1816],[350,-29.1271],[352,-28.7789],[354,-28.179],[358,-28.3983],[360,-28.8618],[362,-28.3844],[364,-26.4999],[366,-17.9194],[368,-13.1412]]
       }
       leg_r_pv_ik_ctrl.Scale.X {type=MovieSceneScriptingFloatChannel, default=1, const [1,140]}
       leg_r_pv_ik_ctrl.Scale.Y {type=MovieSceneScriptingFloatChannel, default=1, const [1,140]}
       leg_r_pv_ik_ctrl.Scale.Z {type=MovieSceneScriptingFloatChannel, default=1, const [1,140]}
       upperarm_l_fk_ctrl.Location.X {type=MovieSceneScriptingFloatChannel, default=0, thinned_from=69, keys [[0,0],[328,0],[342,-0.0005],[346,0],[368,0]]}
       upperarm_l_fk_ctrl.Location.Y {
        type=MovieSceneScriptingFloatChannel
        default=-0.0003
        thinned_from=69
        keys [[0,0],[328,0],[342,-0.0002],[362,0],[366,-0.0004],[368,-0.0003]]
       }
       upperarm_l_fk_ctrl.Location.Z {type=MovieSceneScriptingFloatChannel, default=-0.0003, thinned_from=69, keys [[0,0],[362,0],[364,-0.0002],[366,-0.0011],[368,-0.0003]]}
       upperarm_l_fk_ctrl.Rotation.X {
        type=MovieSceneScriptingFloatChannel
        default=77.635
        keys [[0,100.3896],[92,100.3896],[94,99.386],[96,97.8938],[98,95.941],[100,93.5819],[102,90.8709],[104,87.8624],[114,55.2165],[124,47.1085],[154,6.1878],[194,3.345],[242,0.2348],[244,0.1417],[246,0.0813],[248,-0.069],[250,-0.249],[252,-0.4215],[254,-0.5089],[256,-0.5679],[258,-0.6405],[260,-0.8057],[262,-0.9205],[264,-0.9561],[266,-0.944],[268,-0.8709],[270,-0.906],[272,-0.8405],[274,-0.7186],[276,-0.7315],[278,-0.8877],[280,-1.069],[282,-1.1497],[284,-1.1617],[286,-1.1818],[288,-1.1768],[290,-1.1313],[292,-1.158],[294,-1.1757],[296,-1.1574],[298,-1.1171],[300,-1.0644],[302,-1.0781],[304,-1.0485],[306,-1.0632],[308,-1.0542],[310,-1.0624],[312,-1.0626],[314,-1.0517],[316,-1.0321],[318,-1.0412],[320,-1.0216],[322,-0.9447],[324,-0.7651],[326,-0.6133],[328,-0.6666],[342,-21.3115],[346,-23.9653],[348,-25.9506],[350,-25.607],[352,-24.8578],[354,-22.8332],[356,-21.2612],[358,-19.6371],[360,-14.7292],[362,-5.3751],[364,49.6526],[366,75.7183],[368,77.635]]
       }
       upperarm_l_fk_ctrl.Rotation.Y {
        type=MovieSceneScriptingFloatChannel
        default=31.6141
        thinned_from=69
        keys [[0,39.5545],[92,39.5545],[94,40.3903],[100,43.3298],[102,44.182],[104,44.8643],[114,34.5592],[124,23.2833],[154,5.303],[194,2.9181],[242,1.669],[246,1.5901],[248,1.6145],[250,1.632],[252,1.6269],[254,1.5891],[256,1.5346],[258,1.5469],[260,1.5825],[262,1.6237],[264,1.646],[266,1.6343],[270,1.5533],[272,1.4042],[274,1.2429],[276,1.1311],[278,1.1055],[280,1.1911],[284,1.3776],[286,1.4349],[288,1.3542],[290,1.2053],[292,1.0864],[294,0.9984],[296,0.9686],[298,0.9894],[300,1.0681],[302,1.1444],[304,1.2154],[306,1.2828],[308,1.2803],[312,1.3511],[314,1.3766],[318,1.4869],[320,1.5993],[322,1.6314],[324,1.6974],[326,1.7398],[328,1.7761],[342,21.8468],[346,14.1006],[348,9.85],[350,7.5956],[352,6.5786],[354,6.9446],[356,8.101],[358,11.5509],[360,17.6428],[362,25.8507],[364,49.0058],[366,40.7279],[368,31.6141]]
       }
       upperarm_l_fk_ctrl.Rotation.Z {
        type=MovieSceneScriptingFloatChannel
        default=171.0627
        thinned_from=69
        keys [[0,217.6096],[92,217.6096],[94,215.417],[96,211.9163],[98,206.7836],[100,200.2108],[102,192.39],[104,183.5131],[114,107.82],[124,57.9856],[154,-1.0001],[194,-6.2619],[242,-11.3598],[244,-11.4101],[246,-11.5599],[250,-11.7992],[256,-12.0409],[258,-12.0694],[260,-12.1331],[266,-12.0488],[268,-12.037],[270,-12.2099],[272,-12.5167],[274,-12.7131],[276,-12.8526],[278,-12.9229],[280,-12.957],[286,-13.2194],[288,-13.2784],[290,-13.289],[292,-13.2218],[294,-13.1152],[298,-13.063],[300,-13.0653],[302,-13.139],[304,-13.249],[306,-13.4669],[308,-13.783],[310,-13.945],[316,-14.2394],[318,-14.2794],[322,-14.093],[324,-13.8246],[326,-13.6068],[328,-13.5682],[342,7.8207],[346,6.8563],[348,4.6311],[350,4.2994],[352,4.1443],[354,6.038],[356,8.4473],[358,10.9968],[360,14.5016],[362,20.5557],[364,124.4112],[366,165.7936],[368,171.0627]]
       }
       upperarm_l_fk_ctrl.Scale.X {type=MovieSceneScriptingFloatChannel, default=1, const [1,69]}
       upperarm_l_fk_ctrl.Scale.Y {type=MovieSceneScriptingFloatChannel, default=1, const [1,69]}
       upperarm_l_fk_ctrl.Scale.Z {type=MovieSceneScriptingFloatChannel, default=1, const [1,69]}
       lowerarm_l_fk_ctrl.Location.X {
        type=MovieSceneScriptingFloatChannel
        default=-0.0008
        thinned_from=71
        keys [[0,-0.0001],[94,-0.0002],[104,-0.001],[114,-0.0017],[124,-0.0017],[174,-0.001],[206,-0.0009],[242,-0.0005],[244,-0.0001],[246,-0.0005],[248,0.0004],[250,-0.0003],[252,0.0004],[254,-0.0008],[258,-0.0009],[260,-0.0003],[262,-0.0001],[264,0.0003],[266,-0.0011],[268,0.0004],[270,0.0001],[274,-0.0012],[276,0.0005],[278,0.0007],[280,0.0006],[282,0.0003],[284,-0.0012],[286,0.0007],[288,0.0002],[290,-0.0019],[292,-0.0001],[294,0.0008],[298,-0.0018],[300,-0.0011],[302,-0.0012],[304,-0.001],[306,-0.0001],[308,-0.0029],[310,0.0008],[312,0.0009],[314,-0.0025],[316,-0.0014],[318,-0.0028],[320,-0.0028],[324,-0.0001],[326,-0.0005],[328,0.0005],[342,-0.0042],[346,-0.0011],[348,-0.0013],[352,0],[354,-0.0005],[358,0.0002],[362,0],[364,-0.0005],[366,-0.0097],[368,-0.0008]]
       }
       lowerarm_l_fk_ctrl.Location.Y {
        type=MovieSceneScriptingFloatChannel
        default=0.0004
        thinned_from=71
        keys [[0,-0.0001],[104,-0.0005],[114,-0.001],[124,-0.0012],[242,-0.0009],[244,-0.0003],[246,-0.0008],[248,0.0008],[250,-0.0006],[252,0.0008],[254,-0.0014],[256,-0.0017],[258,-0.0017],[260,-0.0005],[262,-0.0003],[264,0.0005],[266,-0.0019],[268,0.0008],[270,0.0002],[272,-0.0011],[274,-0.0023],[276,0.0009],[278,0.0012],[280,0.001],[282,0.0005],[284,-0.0021],[286,0.0013],[288,0.0004],[290,-0.0034],[292,-0.0002],[294,0.0014],[296,-0.0011],[298,-0.0033],[300,-0.0019],[302,-0.0022],[304,-0.0018],[306,-0.0001],[308,-0.0052],[310,0.0015],[312,0.0016],[314,-0.0046],[316,-0.0026],[318,-0.005],[320,-0.0051],[324,-0.0002],[326,-0.0008],[328,0.0011],[342,0.005],[346,-0.001],[348,-0.0016],[352,0.0001],[354,-0.0006],[358,0.0003],[362,0],[364,0.0002],[366,0.0063],[368,0.0004]]
       }
       lowerarm_l_fk_ctrl.Location.Z {
        type=MovieSceneScriptingFloatChannel
        default=0.0001
        thinned_from=71
        keys [[0,0],[114,-0.0001],[242,0.0004],[244,0.0001],[246,0.0004],[248,-0.0004],[250,0.0003],[252,-0.0004],[254,0.0007],[258,0.0009],[260,0.0002],[262,0.0001],[264,-0.0002],[266,0.001],[268,-0.0004],[270,-0.0001],[274,0.0012],[276,-0.0005],[278,-0.0007],[280,-0.0006],[282,-0.0003],[284,0.0012],[286,-0.0007],[288,-0.0002],[290,0.0019],[292,0.0001],[294,-0.0008],[298,0.0018],[300,0.0011],[302,0.0012],[304,0.001],[306,0.0001],[308,0.003],[310,-0.0009],[312,-0.0009],[314,0.0026],[316,0.0015],[318,0.0029],[320,0.0029],[324,0.0001],[326,0.0004],[328,-0.0006],[342,-0.0019],[346,-0.0001],[352,-0.0001],[354,-0.0002],[358,0],[364,0.0001],[366,-0.0013],[368,0.0001]]
       }
       lowerarm_l_fk_ctrl.Rotation.X {
        type=MovieSceneScriptingFloatChannel
        default=1.1891
        thinned_from=71
        keys [[0,1.0393],[92,1.0393],[94,1.0543],[96,1.103],[98,1.1841],[100,1.2963],[102,1.438],[104,1.608],[114,5.2506],[124,8.6585],[154,11.2023],[174,11.7358],[206,11.9583],[242,12.3878],[270,12.2013],[276,12.2094],[290,12.0961],[306,11.9279],[326,11.7653],[328,11.7287],[342,0],[346,-0.3307],[348,-0.8131],[350,-0.8243],[352,-0.8304],[354,-0.8224],[356,-0.8094],[358,-0.9601],[360,0.323],[362,1.5465],[364,1.1869],[366,0.8842],[368,1.1891]]
       }
       lowerarm_l_fk_ctrl.Rotation.Y {
        type=MovieSceneScriptingFloatChannel
        default=-2.4737
        thinned_from=71
        keys [[0,-2.0106],[92,-2.0106],[94,-2.1228],[96,-2.2475],[98,-2.3824],[100,-2.5251],[104,-2.8245],[114,-4.3907],[124,-5.4433],[154,-5.831],[174,-5.8942],[206,-5.8967],[242,-5.877],[262,-6.0715],[266,-6.1323],[270,-6.172],[276,-6.1849],[290,-6.3708],[306,-6.6197],[318,-6.7743],[326,-6.8899],[328,-6.9354],[342,-12.2492],[346,-13.6711],[348,-13.9217],[356,-13.927],[358,-13.907],[362,0.4977],[364,-1.4431],[366,-4.1857],[368,-2.4737]]
       }
       lowerarm_l_fk_ctrl.Rotation.Z {
        type=MovieSceneScriptingFloatChannel
        default=-18.572
        thinned_from=71
        keys [[0,-20.3979],[92,-20.3979],[94,-19.6276],[96,-18.7902],[100,-16.9883],[104,-15.14],[114,-1.276],[124,14.356],[154,16.76],[194,14.9885],[206,14.9218],[244,17.3081],[246,17.3307],[252,17.6858],[254,17.7726],[256,17.8173],[260,18.0284],[262,18.1672],[264,18.3876],[266,18.555],[270,18.7163],[274,18.5095],[276,18.5082],[280,18.8363],[284,19.0729],[286,19.2892],[290,19.4822],[292,19.718],[294,19.855],[302,20.5237],[306,20.9191],[308,21.0098],[314,21.4594],[316,21.5507],[318,21.7068],[320,21.9108],[324,22.1979],[326,22.306],[328,22.5919],[342,33.6219],[346,29.6072],[348,26.0856],[350,24.2831],[352,22.8833],[354,24.8332],[356,28.3791],[358,32.9839],[360,38.1284],[362,43.751],[364,5.5465],[366,-7.4668],[368,-18.572]]
       }
       lowerarm_l_fk_ctrl.Scale.X {type=MovieSceneScriptingFloatChannel, default=1, const [1,71]}
       lowerarm_l_fk_ctrl.Scale.Y {type=MovieSceneScriptingFloatChannel, default=1, const [1,71]}
       lowerarm_l_fk_ctrl.Scale.Z {type=MovieSceneScriptingFloatChannel, default=1, const [1,71]}
       hand_l_fk_ctrl.Location.X {
        type=MovieSceneScriptingFloatChannel
        default=-0.0003
        thinned_from=77
        keys [[0,-0.0002],[92,-0.0002],[102,-0.0019],[114,-0.0042],[124,-0.0041],[154,-0.0021],[174,-0.0015],[194,-0.0013],[330,0],[332,-0.0124],[334,0],[336,-0.0108],[338,-0.0071],[340,-0.0058],[342,-0.0036],[344,-0.0005],[346,0],[348,-0.0002],[350,0],[354,-0.0002],[356,0],[362,0],[364,-0.0002],[366,-0.0018],[368,-0.0003]]
       }
       hand_l_fk_ctrl.Location.Y {
        type=MovieSceneScriptingFloatChannel
        default=0
        thinned_from=77
        keys [[0,0.0001],[104,0.0005],[114,0.001],[174,0.0003],[330,0],[332,0.0021],[334,0],[336,0.0019],[338,0.0002],[340,-0.0003],[344,0],[346,-0.431],[348,0],[350,-0.431],[352,0],[354,0],[356,-0.431],[358,0],[360,-0.431],[362,-0.431],[364,-0.1512],[366,-0.0008],[368,0]]
       }
       hand_l_fk_ctrl.Location.Z {
        type=MovieSceneScriptingFloatChannel
        default=0
        thinned_from=77
        keys [[0,0],[104,-0.0003],[114,-0.0006],[330,0],[332,-0.0035],[334,0],[336,-0.0013],[340,0.0001],[344,0],[346,0.0668],[348,0],[350,0.0668],[352,0],[354,0],[356,0.0668],[358,0],[360,0.0668],[362,0.0668],[364,0.0234],[366,0.0005],[368,0]]
       }
       hand_l_fk_ctrl.Rotation.X {
        type=MovieSceneScriptingFloatChannel
        default=4.3386
        thinned_from=77
        keys [[0,-2.67],[92,-2.67],[94,-2.3379],[96,-1.9397],[98,-1.4804],[100,-0.965],[102,-0.3984],[104,0.2145],[114,7.0778],[124,15.6927],[154,11.8364],[174,11.9847],[194,12.0285],[242,11.1539],[246,11.3859],[248,11.3011],[250,11.1475],[252,11.2656],[254,11.3173],[258,11.0033],[260,11.051],[262,11.1874],[266,11.485],[268,11.5427],[270,11.403],[272,10.8097],[274,10.2387],[276,9.9741],[278,10.0368],[280,10.298],[282,10.3323],[284,10.1924],[286,10.2468],[288,10.4221],[290,10.5706],[294,11.1],[296,11.182],[298,11.3028],[302,11.0679],[304,10.9219],[306,10.7221],[308,10.4299],[312,10.2228],[316,10.2419],[318,10.2358],[320,10.4307],[322,10.4463],[324,10.3616],[328,10.3557],[330,10.9537],[332,12.1891],[334,13.7957],[336,14.0232],[338,12.152],[340,12.1971],[342,13.7164],[344,15.8681],[346,14.6393],[348,20.6703],[350,16.3146],[352,20.3119],[354,19.9011],[356,17.2037],[358,18.6481],[360,16.7825],[362,0],[364,2.8163],[366,6.7682],[368,4.3386]]
       }
       hand_l_fk_ctrl.Rotation.Y {
        type=MovieSceneScriptingFloatChannel
        default=27.6739
        thinned_from=77
        keys [[0,22.7127],[92,22.7127],[104,22.5719],[114,24.0457],[124,24.3651],[154,29.1298],[174,29.7228],[194,29.9165],[242,30.6027],[244,30.5309],[246,30.6909],[250,30.5624],[252,30.6776],[254,30.6899],[258,30.5637],[260,30.5955],[262,30.7068],[266,30.6643],[268,30.7898],[274,30.4291],[278,30.3556],[280,30.5379],[282,30.5636],[284,30.4803],[286,30.4757],[290,30.7741],[292,30.7083],[294,30.9264],[302,30.7897],[304,30.689],[306,30.5101],[308,30.4163],[320,30.5108],[322,30.4616],[324,30.2788],[326,30.3101],[328,30.1139],[330,30.011],[332,28.9768],[334,25.7468],[336,25.4921],[338,25.6904],[340,25.0277],[342,25.3603],[344,24.9647],[346,23.4231],[348,24.9375],[350,23.9173],[352,25.9606],[354,26.5348],[356,24.2868],[358,27.8052],[360,26.1318],[362,0],[364,17.9637],[366,27.2785],[368,27.6739]]
       }
       hand_l_fk_ctrl.Rotation.Z {
        type=MovieSceneScriptingFloatChannel
        default=-23.7898
        thinned_from=77
        keys [[0,-17.5966],[92,-17.5966],[94,-17.2577],[96,-16.8754],[98,-16.4501],[100,-15.9817],[102,-15.4707],[104,-14.9171],[114,-12.4722],[124,-2.1381],[154,-14.5774],[174,-15.4082],[194,-15.5952],[242,-17.1323],[244,-16.9342],[246,-17.0094],[248,-16.9033],[250,-17.0057],[252,-17.0727],[254,-17.0739],[256,-16.8631],[260,-16.7654],[262,-16.827],[264,-16.6748],[266,-16.5739],[268,-16.5953],[270,-16.4723],[274,-16.7906],[276,-16.7138],[278,-16.6931],[280,-16.7154],[282,-16.8617],[286,-16.7549],[288,-16.5984],[290,-16.5279],[292,-16.2785],[294,-16.252],[296,-16.2774],[298,-16.0239],[300,-16.0048],[302,-16.0349],[304,-16.1247],[308,-16.13],[310,-16.0522],[312,-16.0544],[314,-16.0172],[316,-15.8695],[318,-15.7636],[320,-15.5744],[322,-15.7356],[326,-15.6557],[328,-15.4024],[330,-14.8724],[332,-12.8741],[334,-9.9676],[336,-12.5479],[338,-16.7464],[340,-16.8992],[342,-18.7765],[344,-17.6892],[346,-14.7362],[348,-11.9709],[350,-11.0489],[352,-13.314],[354,-14.1229],[356,-8.9979],[358,-17.7947],[360,-12.2491],[362,0],[364,-15.4425],[366,-22.1752],[368,-23.7898]]
       }
       hand_l_fk_ctrl.Scale.X {type=MovieSceneScriptingFloatChannel, default=1, const [1,77]}
       hand_l_fk_ctrl.Scale.Y {type=MovieSceneScriptingFloatChannel, default=1, const [1,77]}
       hand_l_fk_ctrl.Scale.Z {type=MovieSceneScriptingFloatChannel, default=1, const [1,77]}
       thumb_01_l_ctrl.Location.X {
        type=MovieSceneScriptingFloatChannel
        default=-0.0001
        thinned_from=18
        keys [[0,0],[330,0],[340,-0.0007],[342,-0.0005],[344,0],[358,-0.0001],[360,-0.0002],[362,-0.001],[364,-0.0004],[366,-0.0001],[368,-0.0001]]
       }
       thumb_01_l_ctrl.Location.Y {
        type=MovieSceneScriptingFloatChannel
        default=0.0001
        thinned_from=18
        keys [[0,0],[330,0],[340,-0.0003],[344,0],[360,0],[362,-0.0006],[366,0],[368,0.0001]]
       }
       thumb_01_l_ctrl.Location.Z {
        type=MovieSceneScriptingFloatChannel
        default=0.0001
        thinned_from=18
        keys [[0,0],[342,0.0001],[360,0],[362,-0.0004],[366,0.0001],[368,0.0001]]
       }
       thumb_01_l_ctrl.Rotation.X {
        type=MovieSceneScriptingFloatChannel
        default=-12.8224
        thinned_from=18
        keys [[0,-1.5583],[330,-1.5583],[340,-2.8872],[342,-0.7269],[344,0],[346,-1.4702],[348,-4.0955],[350,-7.1085],[352,-9.8179],[354,-10.5774],[356,-10.3421],[358,-10.076],[360,-9.8719],[362,-9.7269],[364,-11.6442],[366,-12.6548],[368,-12.8224]]
       }
       thumb_01_l_ctrl.Rotation.Y {
        type=MovieSceneScriptingFloatChannel
        default=7.8963
        thinned_from=18
        keys [[0,0.8707],[330,0.8707],[340,2.105],[342,0.4481],[344,0],[346,1.1043],[348,3.3056],[350,5.8691],[352,7.8831],[354,8.6543],[356,8.7266],[358,8.7306],[360,8.6619],[362,8.4444],[364,7.9993],[366,7.8562],[368,7.8963]]
       }
       thumb_01_l_ctrl.Rotation.Z {
        type=MovieSceneScriptingFloatChannel
        default=-11.6866
        thinned_from=18
        keys [[0,-13.9227],[330,-13.9227],[340,-2.836],[342,-1.1329],[344,0],[346,-0.6056],[348,-3.5245],[350,-6.8079],[352,-9.887],[354,-11.1997],[356,-10.945],[358,-10.8838],[360,-11.2629],[362,-11.7384],[364,-11.7542],[366,-11.8289],[368,-11.6866]]
       }
       thumb_01_l_ctrl.Scale.X {type=MovieSceneScriptingFloatChannel, default=1, const [1,18]}
       thumb_01_l_ctrl.Scale.Y {type=MovieSceneScriptingFloatChannel, default=1, const [1,18]}
       thumb_01_l_ctrl.Scale.Z {type=MovieSceneScriptingFloatChannel, default=1, const [1,18]}
       thumb_02_l_ctrl.Location.X {
        type=MovieSceneScriptingFloatChannel
        default=0.0002
        thinned_from=18
        keys [[0,0],[330,0],[340,-0.0015],[342,-0.0012],[344,0],[358,0.0002],[360,-0.0001],[362,-0.0011],[364,0.0002],[366,0],[368,0.0002]]
       }
       thumb_02_l_ctrl.Location.Y {
        type=MovieSceneScriptingFloatChannel
        default=0
        thinned_from=18
        keys [[0,0],[330,0],[340,-0.001],[342,-0.0007],[344,0],[358,0],[360,-0.0002],[362,-0.0012],[364,-0.0003],[366,0.0001],[368,0]]
       }
       thumb_02_l_ctrl.Location.Z {
        type=MovieSceneScriptingFloatChannel
        default=0.0001
        thinned_from=18
        keys [[0,0],[330,0],[342,0.0003],[344,0],[360,0],[362,-0.0011],[364,-0.0001],[366,0.0002],[368,0.0001]]
       }
       thumb_02_l_ctrl.Rotation.X {
        type=MovieSceneScriptingFloatChannel
        default=-0.6563
        thinned_from=18
        keys [[0,0],[330,0],[340,0.3706],[342,0.2032],[344,0],[346,-0.2062],[348,-0.2535],[350,-0.2659],[352,-0.283],[354,-0.2226],[356,-0.1608],[358,-0.1008],[360,-0.0586],[362,0.018],[364,-0.3677],[366,-0.5734],[368,-0.6563]]
       }
       thumb_02_l_ctrl.Rotation.Y {
        type=MovieSceneScriptingFloatChannel
        default=2.9372
        thinned_from=18
        keys [[0,0],[330,0],[340,-1.6892],[342,-0.7514],[344,0],[346,0.9414],[348,1.4196],[350,1.8367],[352,2.0951],[354,1.9405],[356,1.6499],[358,1.3397],[360,1.165],[362,0.8176],[364,1.8184],[366,2.7954],[368,2.9372]]
       }
       thumb_02_l_ctrl.Rotation.Z {
        type=MovieSceneScriptingFloatChannel
        default=-7.7152
        thinned_from=18
        keys [[0,0],[330,0],[340,-3.9128],[342,-0.8922],[344,0],[346,-0.277],[348,-1.8142],[350,-3.7069],[352,-5.5361],[354,-6.545],[356,-6.8419],[358,-7.2204],[360,-7.3435],[362,-7.7156],[364,-8.4952],[366,-7.3731],[368,-7.7152]]
       }
       thumb_02_l_ctrl.Scale.X {type=MovieSceneScriptingFloatChannel, default=1, const [1,18]}
       thumb_02_l_ctrl.Scale.Y {type=MovieSceneScriptingFloatChannel, default=1, const [1,18]}
       thumb_02_l_ctrl.Scale.Z {type=MovieSceneScriptingFloatChannel, default=1, const [1,18]}
       thumb_03_l_ctrl.Location.X {
        type=MovieSceneScriptingFloatChannel
        default=0
        thinned_from=18
        keys [[0,0],[330,0],[340,-0.0006],[344,0],[358,0],[360,-0.0001],[362,-0.0007],[364,-0.0001],[368,0]]
       }
       thumb_03_l_ctrl.Location.Y {type=MovieSceneScriptingFloatChannel, default=0, thinned_from=18, keys [[0,0],[360,0],[362,-0.0003],[366,0.0001],[368,0]]}
       thumb_03_l_ctrl.Location.Z {type=MovieSceneScriptingFloatChannel, default=0, thinned_from=18, keys [[0,0],[342,0.0001],[360,0],[362,-0.0004],[366,0],[368,0]]}
       thumb_03_l_ctrl.Rotation.X {
        type=MovieSceneScriptingFloatChannel
        default=1.6608
        thinned_from=18
        keys [[0,8.1645],[330,8.1645],[340,0.2696],[342,0.1982],[344,0],[346,0.3857],[348,0.9009],[350,1.424],[352,1.903],[354,2.0016],[356,1.9837],[358,1.8495],[360,1.8224],[362,1.8553],[364,1.5423],[366,1.7922],[368,1.6608]]
       }
       thumb_03_l_ctrl.Rotation.Y {
        type=MovieSceneScriptingFloatChannel
        default=-0.5289
        thinned_from=18
        keys [[0,-9.9357],[330,-9.9357],[340,-0.235],[342,0.1914],[344,0],[346,0.3845],[348,0.4623],[350,0.3723],[352,0.376],[354,0.1771],[356,0.1421],[358,-0.1299],[360,-0.1857],[362,-0.1184],[364,-0.8005],[366,-0.2578],[368,-0.5289]]
       }
       thumb_03_l_ctrl.Rotation.Z {
        type=MovieSceneScriptingFloatChannel
        default=-54.3766
        thinned_from=18
        keys [[0,-50.7089],[330,-50.7089],[340,-11.1551],[342,-2.3687],[344,0],[346,-4.3727],[348,-17.2458],[350,-33.1485],[352,-46.2381],[354,-52.3193],[356,-52.4225],[358,-53.2194],[360,-53.3823],[362,-53.1864],[364,-55.1442],[366,-53.5855],[368,-54.3766]]
       }
       thumb_03_l_ctrl.Scale.X {type=MovieSceneScriptingFloatChannel, default=1, const [1,18]}
       thumb_03_l_ctrl.Scale.Y {type=MovieSceneScriptingFloatChannel, default=1, const [1,18]}
       thumb_03_l_ctrl.Scale.Z {type=MovieSceneScriptingFloatChannel, default=1, const [1,18]}
       index_metacarpal_l_ctrl.Location.X {
        type=MovieSceneScriptingFloatChannel
        default=0.0001
        thinned_from=18
        keys [[0,0],[330,0],[340,-0.0003],[358,0.0001],[360,0],[362,-0.0006],[366,0.0002],[368,0.0001]]
       }
       index_metacarpal_l_ctrl.Location.Y {
        type=MovieSceneScriptingFloatChannel
        default=0.0001
        thinned_from=18
        keys [[0,0],[330,0],[340,0.0005],[342,0.0004],[344,0],[362,0.0002],[368,0.0001]]
       }
       index_metacarpal_l_ctrl.Location.Z {
        type=MovieSceneScriptingFloatChannel
        default=0
        thinned_from=18
        keys [[0,0],[330,0],[340,0.0006],[344,0],[360,0.0002],[362,0.0012],[366,0.0001],[368,0]]
       }
       index_metacarpal_l_ctrl.Rotation.X {type=MovieSceneScriptingFloatChannel, default=0, thinned_from=18, keys [[0,0],[368,0]]}
       index_metacarpal_l_ctrl.Rotation.Y {
        type=MovieSceneScriptingFloatChannel
        default=4.1035
        thinned_from=18
        keys [[0,3.7815],[330,3.7815],[340,0.8135],[342,0.2273],[344,0],[346,0.4268],[348,1.4444],[350,2.6591],[352,3.6768],[354,4.1035],[368,4.1035]]
       }
       index_metacarpal_l_ctrl.Rotation.Z {type=MovieSceneScriptingFloatChannel, default=0, thinned_from=18, keys [[0,-3.094],[330,-3.094],[340,0],[368,0]]}
       index_metacarpal_l_ctrl.Scale.X {type=MovieSceneScriptingFloatChannel, default=1, const [1,18]}
       index_metacarpal_l_ctrl.Scale.Y {type=MovieSceneScriptingFloatChannel, default=1, const [1,18]}
       index_metacarpal_l_ctrl.Scale.Z {type=MovieSceneScriptingFloatChannel, default=1, const [1,18]}
       index_01_l_ctrl.Location.X {
        type=MovieSceneScriptingFloatChannel
        default=0.0001
        thinned_from=18
        keys [[0,0],[330,0],[340,-0.0024],[342,-0.0017],[344,0],[352,0.0001],[356,0],[358,0.0001],[360,-0.0004],[362,-0.0032],[364,-0.0005],[368,0.0001]]
       }
       index_01_l_ctrl.Location.Y {
        type=MovieSceneScriptingFloatChannel
        default=0
        thinned_from=18
        keys [[0,0],[330,0],[342,0.0004],[344,0],[360,-0.0001],[362,-0.0012],[364,-0.0009],[366,0],[368,0]]
       }
       index_01_l_ctrl.Location.Z {
        type=MovieSceneScriptingFloatChannel
        default=0.0001
        thinned_from=18
        keys [[0,0],[330,0],[342,-0.0003],[344,0],[360,0],[362,0.0005],[364,0.0004],[366,-0.0002],[368,0.0001]]
       }
       index_01_l_ctrl.Rotation.X {
        type=MovieSceneScriptingFloatChannel
        default=1.425
        thinned_from=18
        keys [[0,6.2821],[330,6.2821],[340,-1.8735],[342,-0.6992],[344,0],[346,0.9231],[348,1.7949],[350,2.6863],[352,3.2923],[354,3.5283],[356,3.5597],[358,3.566],[360,3.459],[362,3.3288],[364,1.7697],[366,1.4459],[368,1.425]]
       }
       index_01_l_ctrl.Rotation.Y {
        type=MovieSceneScriptingFloatChannel
        default=-1.6015
        thinned_from=18
        keys [[0,-2.5276],[330,-2.5276],[340,-0.4913],[342,-0.0677],[344,0],[346,-0.3402],[348,-0.9604],[350,-1.577],[352,-2.1072],[354,-2.2587],[356,-2.2618],[358,-2.1864],[360,-2.1354],[362,-1.9695],[364,-1.6566],[366,-1.6256],[368,-1.6015]]
       }
       index_01_l_ctrl.Rotation.Z {
        type=MovieSceneScriptingFloatChannel
        default=-36.3486
        thinned_from=18
        keys [[0,-65.1155],[330,-65.1155],[340,1.3468],[342,0.9923],[344,0],[346,-6.8804],[348,-18.2789],[350,-31.0721],[352,-41.1928],[354,-45.2353],[358,-45.1293],[360,-44.781],[362,-44.2402],[364,-37.6628],[366,-36.5871],[368,-36.3486]]
       }
       index_01_l_ctrl.Scale.X {type=MovieSceneScriptingFloatChannel, default=1, const [1,18]}
       index_01_l_ctrl.Scale.Y {type=MovieSceneScriptingFloatChannel, default=1, const [1,18]}
       index_01_l_ctrl.Scale.Z {type=MovieSceneScriptingFloatChannel, default=1, const [1,18]}
       index_02_l_ctrl.Location.X {
        type=MovieSceneScriptingFloatChannel
        default=-0.0001
        thinned_from=18
        keys [[0,0],[330,0],[342,-0.0003],[344,0],[362,0],[364,-0.0003],[368,-0.0001]]
       }
       index_02_l_ctrl.Location.Y {
        type=MovieSceneScriptingFloatChannel
        default=-0.0001
        thinned_from=18
        keys [[0,0],[330,0],[340,-0.0006],[342,-0.0004],[344,0],[362,-0.0003],[364,0],[368,-0.0001]]
       }
       index_02_l_ctrl.Location.Z {type=MovieSceneScriptingFloatChannel, default=-0.0001, thinned_from=18, keys [[0,0],[360,0],[362,0.0001],[364,-0.0001],[368,-0.0001]]}
       index_02_l_ctrl.Rotation.X {
        type=MovieSceneScriptingFloatChannel
        default=0.284
        thinned_from=18
        keys [[0,0],[330,0],[340,1.5191],[342,0.5121],[344,0],[346,-0.3176],[348,-0.3585],[350,-0.3689],[352,-0.3024],[354,-0.2795],[356,-0.2876],[358,-0.3165],[360,-0.2785],[362,-0.2999],[364,0.127],[366,0.34],[368,0.284]]
       }
       index_02_l_ctrl.Rotation.Y {
        type=MovieSceneScriptingFloatChannel
        default=-0.345
        thinned_from=18
        keys [[0,0],[330,0],[340,-0.4679],[342,-0.1249],[344,0],[346,0.0158],[348,-0.0571],[350,-0.1512],[352,-0.238],[354,-0.274],[356,-0.2732],[358,-0.2707],[360,-0.2742],[362,-0.2726],[364,-0.324],[366,-0.3546],[368,-0.345]]
       }
       index_02_l_ctrl.Rotation.Z {
        type=MovieSceneScriptingFloatChannel
        default=-74.8134
        thinned_from=18
        keys [[0,-60.8298],[330,-60.8298],[340,-4.0791],[342,-0.5495],[344,0],[346,-10.2676],[348,-29.6139],[350,-52.4297],[352,-71.0098],[354,-78.8369],[358,-79.1001],[360,-78.8304],[362,-78.9872],[364,-75.9442],[366,-74.411],[368,-74.8134]]
       }
       index_02_l_ctrl.Scale.X {type=MovieSceneScriptingFloatChannel, default=1, const [1,18]}
       index_02_l_ctrl.Scale.Y {type=MovieSceneScriptingFloatChannel, default=1, const [1,18]}
       index_02_l_ctrl.Scale.Z {type=MovieSceneScriptingFloatChannel, default=1, const [1,18]}
       index_03_l_ctrl.Location.X {
        type=MovieSceneScriptingFloatChannel
        default=-0.0001
        thinned_from=18
        keys [[0,0],[358,0],[360,-0.0001],[362,-0.0009],[366,-0.0001],[368,-0.0001]]
       }
       index_03_l_ctrl.Location.Y {type=MovieSceneScriptingFloatChannel, default=0, thinned_from=18, keys [[0,0],[330,0],[342,-0.0001],[362,0.0001],[364,-0.0002],[368,0]]}
       index_03_l_ctrl.Location.Z {type=MovieSceneScriptingFloatChannel, default=-0.0001, thinned_from=18, keys [[0,0],[330,0],[342,-0.0001],[368,-0.0001]]}
       index_03_l_ctrl.Rotation.X {
        type=MovieSceneScriptingFloatChannel
        default=-0.5117
        thinned_from=18
        keys [[0,0.0236],[330,0.0236],[340,0.38],[342,0.1584],[344,0],[346,-0.1509],[348,-0.3134],[350,-0.5146],[352,-0.633],[354,-0.6984],[356,-0.7152],[358,-0.7389],[360,-0.7182],[362,-0.7086],[364,-0.5722],[366,-0.488],[368,-0.5117]]
       }
       index_03_l_ctrl.Rotation.Y {
        type=MovieSceneScriptingFloatChannel
        default=0.1307
        thinned_from=18
        keys [[0,-0.109],[330,-0.109],[340,-0.7719],[342,-0.2777],[344,0],[346,0.1519],[348,0.2338],[350,0.3406],[352,0.3624],[354,0.3926],[356,0.4151],[358,0.4467],[360,0.4193],[362,0.4072],[364,0.2191],[366,0.0952],[368,0.1307]]
       }
       index_03_l_ctrl.Rotation.Z {
        type=MovieSceneScriptingFloatChannel
        default=-26.9149
        thinned_from=19
        keys [[0,-78.806],[330,-78.806],[340,1.3959],[342,0.9594],[344,0],[346,-4.0267],[348,-11.143],[350,-19.7222],[352,-26.2994],[354,-29.248],[356,-29.4515],[358,-29.7387],[360,-29.4908],[362,-29.3812],[364,-27.6936],[366,-26.6056],[368,-26.9149]]
       }
       index_03_l_ctrl.Scale.X {type=MovieSceneScriptingFloatChannel, default=1, const [1,18]}
       index_03_l_ctrl.Scale.Y {type=MovieSceneScriptingFloatChannel, default=1, const [1,18]}
       index_03_l_ctrl.Scale.Z {type=MovieSceneScriptingFloatChannel, default=1, const [1,18]}
       middle_metacarpal_l_ctrl.Location.X {
        type=MovieSceneScriptingFloatChannel
        default=0
        thinned_from=18
        keys [[0,0],[330,0],[340,-0.0006],[344,0],[360,-0.0001],[362,-0.0009],[364,-0.0002],[368,0]]
       }
       middle_metacarpal_l_ctrl.Location.Y {
        type=MovieSceneScriptingFloatChannel
        default=0.0001
        thinned_from=18
        keys [[0,0],[330,0],[342,0.0003],[344,0],[360,0.0001],[364,-0.0001],[368,0.0001]]
       }
       middle_metacarpal_l_ctrl.Location.Z {type=MovieSceneScriptingFloatChannel, default=0, thinned_from=18, keys [[0,0],[360,0],[362,0.0003],[366,0],[368,0]]}
       middle_metacarpal_l_ctrl.Rotation.X {type=MovieSceneScriptingFloatChannel, default=0, thinned_from=18, keys [[0,-7.29],[330,-7.29],[340,0],[368,0]]}
       middle_metacarpal_l_ctrl.Rotation.Y {
        type=MovieSceneScriptingFloatChannel
        default=1.2252
        thinned_from=18
        keys [[0,1.3582],[330,1.3582],[340,0.2429],[342,0.0679],[344,0],[346,0.1274],[348,0.4313],[350,0.794],[352,1.0978],[354,1.2252],[368,1.2252]]
       }
       middle_metacarpal_l_ctrl.Rotation.Z {type=MovieSceneScriptingFloatChannel, default=0, thinned_from=18, keys [[0,-2.7048],[330,-2.7048],[340,0],[368,0]]}
       middle_metacarpal_l_ctrl.Scale.X {type=MovieSceneScriptingFloatChannel, default=1, const [1,18]}
       middle_metacarpal_l_ctrl.Scale.Y {type=MovieSceneScriptingFloatChannel, default=1, const [1,18]}
       middle_metacarpal_l_ctrl.Scale.Z {type=MovieSceneScriptingFloatChannel, default=1, const [1,18]}
       middle_01_l_ctrl.Location.X {
        type=MovieSceneScriptingFloatChannel
        default=0.0002
        thinned_from=18
        keys [[0,0],[330,0],[340,-0.0021],[342,-0.0016],[344,0],[352,0.0003],[358,0.0003],[360,-0.0002],[362,-0.0027],[364,-0.0002],[368,0.0002]]
       }
       middle_01_l_ctrl.Location.Y {
        type=MovieSceneScriptingFloatChannel
        default=0
        thinned_from=18
        keys [[0,0],[330,0],[342,0.0003],[344,0],[360,-0.0001],[362,-0.0011],[364,-0.0007],[366,-0.0001],[368,0]]
       }
       middle_01_l_ctrl.Location.Z {
        type=MovieSceneScriptingFloatChannel
        default=0.0001
        thinned_from=18
        keys [[0,0],[330,0],[342,-0.0003],[344,0],[360,0],[362,0.0003],[364,0.0003],[366,-0.0002],[368,0.0001]]
       }
       middle_01_l_ctrl.Rotation.X {
        type=MovieSceneScriptingFloatChannel
        default=4.0731
        thinned_from=18
        keys [[0,0],[330,0],[340,1.0539],[342,0.304],[344,0],[346,0.3559],[348,1.3115],[350,2.4672],[352,3.4532],[354,3.8649],[360,3.8747],[362,3.89],[364,4.0488],[366,4.0692],[368,4.0731]]
       }
       middle_01_l_ctrl.Rotation.Y {
        type=MovieSceneScriptingFloatChannel
        default=-2.6749
        thinned_from=18
        keys [[0,1.3494],[330,1.3494],[340,1.9839],[342,0.6832],[344,0],[346,-0.9764],[348,-1.9932],[350,-2.9785],[352,-3.7129],[354,-3.9845],[356,-4.0682],[358,-3.9968],[360,-3.8514],[362,-3.8007],[364,-2.7453],[366,-2.5705],[368,-2.6749]]
       }
       middle_01_l_ctrl.Rotation.Z {
        type=MovieSceneScriptingFloatChannel
        default=-52.5728
        thinned_from=19
        keys [[0,-74.7137],[330,-74.7137],[340,-0.03],[342,0.6264],[344,0],[346,-8.502],[348,-23.7919],[350,-41.2838],[352,-55.1844],[354,-61.04],[356,-61.0025],[360,-60.5657],[362,-59.9457],[364,-53.5117],[366,-52.58],[368,-52.5728]]
       }
       middle_01_l_ctrl.Scale.X {type=MovieSceneScriptingFloatChannel, default=1, const [1,18]}
       middle_01_l_ctrl.Scale.Y {type=MovieSceneScriptingFloatChannel, default=1, const [1,18]}
       middle_01_l_ctrl.Scale.Z {type=MovieSceneScriptingFloatChannel, default=1, const [1,18]}
       middle_02_l_ctrl.Location.X {
        type=MovieSceneScriptingFloatChannel
        default=0.0008
        thinned_from=18
        keys [[0,-0.0452],[330,-0.0452],[340,0.0001],[344,0],[354,0.0008],[362,0.0006],[364,0.0003],[368,0.0008]]
       }
       middle_02_l_ctrl.Location.Y {
        type=MovieSceneScriptingFloatChannel
        default=0.0008
        thinned_from=18
        keys [[0,0.0507],[330,0.0507],[340,-0.0003],[342,-0.0003],[344,0],[346,0.0001],[352,0.0008],[358,0.0009],[364,0.0009],[368,0.0008]]
       }
       middle_02_l_ctrl.Location.Z {
        type=MovieSceneScriptingFloatChannel
        default=0.0476
        thinned_from=18
        keys [[0,0.1146],[330,0.1146],[340,0.0096],[342,0.0028],[344,0],[346,0.0049],[348,0.0167],[350,0.0308],[352,0.0426],[354,0.0476],[360,0.0477],[362,0.048],[364,0.0475],[368,0.0476]]
       }
       middle_02_l_ctrl.Rotation.X {
        type=MovieSceneScriptingFloatChannel
        default=-2.7708
        thinned_from=18
        keys [[0,-3.3929],[330,-3.3929],[340,1.2442],[342,0.4536],[344,0],[346,-0.6814],[348,-1.4433],[350,-2.2788],[352,-2.8718],[354,-3.1485],[358,-3.2482],[360,-3.1835],[362,-3.1969],[364,-2.9235],[366,-2.6824],[368,-2.7708]]
       }
       middle_02_l_ctrl.Rotation.Y {
        type=MovieSceneScriptingFloatChannel
        default=-2.3223
        thinned_from=18
        keys [[0,-1.8646],[330,-1.8646],[340,-1.2266],[342,-0.3372],[344,0],[346,-0.1375],[348,-0.7024],[350,-1.3993],[352,-2.0039],[354,-2.2531],[358,-2.2382],[360,-2.2479],[362,-2.2459],[366,-2.3421],[368,-2.3223]]
       }
       middle_02_l_ctrl.Rotation.Z {
        type=MovieSceneScriptingFloatChannel
        default=-61.6077
        thinned_from=18
        keys [[0,-58.0319],[330,-58.0319],[340,2.6419],[342,1.4498],[344,0],[346,-9.5037],[348,-25.3519],[350,-43.6869],[352,-58.2194],[354,-64.529],[358,-65.2966],[360,-64.8002],[362,-64.9099],[364,-62.8],[366,-60.9214],[368,-61.6077]]
       }
       middle_02_l_ctrl.Scale.X {type=MovieSceneScriptingFloatChannel, default=1, const [1,18]}
       middle_02_l_ctrl.Scale.Y {type=MovieSceneScriptingFloatChannel, default=1, const [1,18]}
       middle_02_l_ctrl.Scale.Z {type=MovieSceneScriptingFloatChannel, default=1, const [1,18]}
       middle_03_l_ctrl.Location.X {
        type=MovieSceneScriptingFloatChannel
        default=0.0005
        thinned_from=18
        keys [[0,0.0206],[330,0.0206],[340,0],[346,0],[358,0.0005],[360,0.0004],[362,-0.0005],[364,0.0001],[366,0.0005],[368,0.0005]]
       }
       middle_03_l_ctrl.Location.Y {type=MovieSceneScriptingFloatChannel, default=0, thinned_from=18, keys [[0,-0.0583],[330,-0.0583],[340,0],[362,0],[364,-0.0003],[368,0]]}
       middle_03_l_ctrl.Location.Z {
        type=MovieSceneScriptingFloatChannel
        default=0.0061
        thinned_from=18
        keys [[0,0.1324],[330,0.1324],[340,0.001],[342,0.0001],[344,0],[346,0.0006],[348,0.0022],[350,0.004],[352,0.0055],[354,0.0061],[368,0.0061]]
       }
       middle_03_l_ctrl.Rotation.X {
        type=MovieSceneScriptingFloatChannel
        default=-1.2527
        thinned_from=18
        keys [[0,-1.8232],[330,-1.8232],[340,0.6298],[342,0.2474],[344,0],[346,-0.3581],[348,-0.7215],[350,-1.1356],[352,-1.4138],[354,-1.5606],[356,-1.6126],[358,-1.6289],[360,-1.5381],[362,-1.5447],[364,-1.3307],[366,-1.2011],[368,-1.2527]]
       }
       middle_03_l_ctrl.Rotation.Y {
        type=MovieSceneScriptingFloatChannel
        default=1.5397
        thinned_from=18
        keys [[0,1.4852],[330,1.4852],[340,-0.4889],[342,-0.173],[344,0],[346,0.3292],[348,0.7469],[350,1.2305],[352,1.5886],[354,1.7592],[356,1.794],[358,1.8048],[360,1.7441],[362,1.7491],[364,1.5986],[366,1.4997],[368,1.5397]]
       }
       middle_03_l_ctrl.Rotation.Z {
        type=MovieSceneScriptingFloatChannel
        default=-52.212
        thinned_from=18
        keys [[0,-64.6207],[330,-64.6207],[340,-1.3877],[342,0.1853],[344,0],[346,-7.5856],[348,-21.0159],[350,-36.8626],[352,-49.5058],[354,-55.0852],[356,-55.5602],[358,-55.7091],[360,-54.8805],[362,-54.944],[364,-52.9517],[366,-51.7166],[368,-52.212]]
       }
       middle_03_l_ctrl.Scale.X {type=MovieSceneScriptingFloatChannel, default=1, const [1,18]}
       middle_03_l_ctrl.Scale.Y {type=MovieSceneScriptingFloatChannel, default=1, const [1,18]}
       middle_03_l_ctrl.Scale.Z {type=MovieSceneScriptingFloatChannel, default=1, const [1,18]}
       ring_metacarpal_l_ctrl.Location.X {
        type=MovieSceneScriptingFloatChannel
        default=0.0001
        thinned_from=18
        keys [[0,0],[330,0],[340,-0.0007],[342,-0.0006],[344,0],[358,0.0001],[360,-0.0001],[362,-0.0009],[364,0],[368,0.0001]]
       }
       ring_metacarpal_l_ctrl.Location.Y {
        type=MovieSceneScriptingFloatChannel
        default=0.0002
        thinned_from=18
        keys [[0,0],[330,0],[342,0.0001],[360,0.0001],[362,-0.0001],[366,0.0002],[368,0.0002]]
       }
       ring_metacarpal_l_ctrl.Location.Z {
        type=MovieSceneScriptingFloatChannel
        default=0
        thinned_from=18
        keys [[0,0],[330,0],[342,-0.0002],[344,0],[360,-0.0001],[362,-0.0004],[364,-0.0001],[368,0]]
       }
       ring_metacarpal_l_ctrl.Rotation.X {type=MovieSceneScriptingFloatChannel, default=0, thinned_from=18, keys [[0,-2.4847],[330,-2.4847],[340,0],[368,0]]}
       ring_metacarpal_l_ctrl.Rotation.Y {
        type=MovieSceneScriptingFloatChannel
        default=-4.4979
        thinned_from=18
        keys [[0,-0.8637],[330,-0.8637],[340,-0.8917],[342,-0.2492],[344,0],[346,-0.4678],[348,-1.5833],[350,-2.9146],[352,-4.0301],[354,-4.4979],[368,-4.4979]]
       }
       ring_metacarpal_l_ctrl.Rotation.Z {type=MovieSceneScriptingFloatChannel, default=0, thinned_from=18, keys [[0,-5.6698],[330,-5.6698],[340,0],[368,0]]}
       ring_metacarpal_l_ctrl.Scale.X {type=MovieSceneScriptingFloatChannel, default=1, const [1,18]}
       ring_metacarpal_l_ctrl.Scale.Y {type=MovieSceneScriptingFloatChannel, default=1, const [1,18]}
       ring_metacarpal_l_ctrl.Scale.Z {type=MovieSceneScriptingFloatChannel, default=1, const [1,18]}
       ring_01_l_ctrl.Location.X {
        type=MovieSceneScriptingFloatChannel
        default=-0.0001
        thinned_from=18
        keys [[0,0],[330,0],[340,-0.0022],[342,-0.0016],[344,0],[356,-0.0002],[358,-0.0001],[360,-0.0006],[362,-0.0027],[364,-0.0004],[368,-0.0001]]
       }
       ring_01_l_ctrl.Location.Y {
        type=MovieSceneScriptingFloatChannel
        default=0
        thinned_from=18
        keys [[0,0],[330,0],[340,-0.0003],[346,0],[358,0],[360,-0.0002],[362,-0.0015],[366,-0.0001],[368,0]]
       }
       ring_01_l_ctrl.Location.Z {
        type=MovieSceneScriptingFloatChannel
        default=0.0001
        thinned_from=18
        keys [[0,0],[330,0],[342,-0.0004],[344,0],[360,0],[362,0.0003],[364,0.0003],[366,-0.0002],[368,0.0001]]
       }
       ring_01_l_ctrl.Rotation.X {
        type=MovieSceneScriptingFloatChannel
        default=-3.851
        thinned_from=18
        keys [[0,-4.509],[330,-4.509],[340,-1.1472],[342,-0.3512],[344,0],[346,-0.2073],[348,-1.0159],[350,-2.1592],[352,-3.083],[354,-3.5059],[358,-3.5056],[360,-3.5265],[362,-3.4954],[364,-3.7535],[366,-3.8242],[368,-3.851]]
       }
       ring_01_l_ctrl.Rotation.Y {
        type=MovieSceneScriptingFloatChannel
        default=5.4693
        thinned_from=18
        keys [[0,-1.201],[330,-1.201],[340,2.3871],[342,0.7572],[344,0],[346,0.5556],[348,1.6367],[350,2.7555],[352,4.0144],[354,4.4203],[356,4.5282],[358,4.6946],[360,4.5453],[362,4.6468],[364,5.4918],[366,5.499],[368,5.4693]]
       }
       ring_01_l_ctrl.Rotation.Z {
        type=MovieSceneScriptingFloatChannel
        default=-51.7016
        thinned_from=19
        keys [[0,-79.0231],[330,-79.0231],[340,-2.2941],[342,0.2288],[344,0],[346,-8.0963],[348,-23.6038],[350,-40.7019],[352,-54.474],[354,-60.2289],[356,-60.1049],[360,-59.4347],[362,-58.8093],[364,-52.4688],[366,-51.6007],[368,-51.7016]]
       }
       ring_01_l_ctrl.Scale.X {type=MovieSceneScriptingFloatChannel, default=1, const [1,18]}
       ring_01_l_ctrl.Scale.Y {type=MovieSceneScriptingFloatChannel, default=1, const [1,18]}
       ring_01_l_ctrl.Scale.Z {type=MovieSceneScriptingFloatChannel, default=1, const [1,18]}
       ring_02_l_ctrl.Location.X {type=MovieSceneScriptingFloatChannel, default=0.0002, thinned_from=18, keys [[0,0],[330,0],[342,-0.0002],[362,0.0002],[364,0],[368,0.0002]]}
       ring_02_l_ctrl.Location.Y {
        type=MovieSceneScriptingFloatChannel
        default=0
        thinned_from=18
        keys [[0,0],[330,0],[340,-0.0005],[344,0],[362,-0.0001],[364,0.0002],[368,0]]
       }
       ring_02_l_ctrl.Location.Z {
        type=MovieSceneScriptingFloatChannel
        default=0
        thinned_from=18
        keys [[0,0],[330,0],[340,0.0003],[358,0],[360,0.0001],[362,0.0004],[364,-0.0001],[368,0]]
       }
       ring_02_l_ctrl.Rotation.X {
        type=MovieSceneScriptingFloatChannel
        default=0.808
        thinned_from=18
        keys [[0,0],[330,0],[340,2.6594],[342,0.837],[344,0],[346,-0.6767],[348,-0.5602],[350,-0.4148],[352,-0.1546],[354,0],[356,-0.1036],[358,-0.162],[360,0.1727],[362,0.1318],[364,0.5904],[366,0.9104],[368,0.808]]
       }
       ring_02_l_ctrl.Rotation.Y {
        type=MovieSceneScriptingFloatChannel
        default=0.0155
        thinned_from=18
        keys [[0,0],[330,0],[340,-0.5563],[342,-0.1082],[344,0],[346,0.036],[348,0.0295],[350,0.0126],[352,0.0019],[354,0],[356,-0.0069],[358,-0.0116],[360,0.0089],[362,0.0069],[364,0.014],[366,0.0114],[368,0.0155]]
       }
       ring_02_l_ctrl.Rotation.Z {
        type=MovieSceneScriptingFloatChannel
        default=-66.1738
        thinned_from=18
        keys [[0,-66.6687],[330,-66.6687],[340,2.9358],[342,1.3182],[344,0],[346,-11.649],[348,-28.5799],[350,-48.7532],[352,-64.7902],[354,-71.2302],[356,-71.8782],[358,-72.2458],[360,-70.152],[362,-70.4142],[364,-67.5434],[366,-65.5342],[368,-66.1738]]
       }
       ring_02_l_ctrl.Scale.X {type=MovieSceneScriptingFloatChannel, default=1, const [1,18]}
       ring_02_l_ctrl.Scale.Y {type=MovieSceneScriptingFloatChannel, default=1, const [1,18]}
       ring_02_l_ctrl.Scale.Z {type=MovieSceneScriptingFloatChannel, default=1, const [1,18]}
       ring_03_l_ctrl.Location.X {
        type=MovieSceneScriptingFloatChannel
        default=0
        thinned_from=18
        keys [[0,0],[358,0],[360,-0.0001],[362,-0.001],[364,-0.0004],[366,-0.0001],[368,0]]
       }
       ring_03_l_ctrl.Location.Y {type=MovieSceneScriptingFloatChannel, default=-0.0001, thinned_from=18, keys [[0,0],[362,-0.0002],[364,-0.0004],[368,-0.0001]]}
       ring_03_l_ctrl.Location.Z {type=MovieSceneScriptingFloatChannel, default=0, thinned_from=18, keys [[0,0],[368,0]]}
       ring_03_l_ctrl.Rotation.X {
        type=MovieSceneScriptingFloatChannel
        default=0.379
        thinned_from=18
        keys [[0,0],[330,0],[340,1.0013],[342,0.3651],[344,0],[346,-0.2704],[348,-0.3011],[350,-0.2009],[352,-0.0625],[354,0],[356,-0.0867],[358,-0.0707],[360,0.1091],[362,0.0686],[364,0.2801],[366,0.4296],[368,0.379]]
       }
       ring_03_l_ctrl.Rotation.Y {
        type=MovieSceneScriptingFloatChannel
        default=-0.2755
        thinned_from=18
        keys [[0,0],[330,0],[340,-0.8927],[342,-0.3039],[344,0],[346,0.2086],[348,0.2267],[350,0.1491],[352,0.0462],[354,0],[356,0.0591],[358,0.0483],[360,-0.076],[362,-0.0466],[364,-0.2009],[366,-0.3149],[368,-0.2755]]
       }
       ring_03_l_ctrl.Rotation.Z {
        type=MovieSceneScriptingFloatChannel
        default=-52.2902
        thinned_from=18
        keys [[0,-73.2004],[330,-73.2004],[340,-3.0355],[342,-0.2666],[344,0],[346,-7.7352],[348,-21.5932],[350,-37.1432],[352,-49.7809],[354,-55.0497],[356,-55.6672],[358,-55.5532],[360,-54.2693],[362,-54.5644],[364,-53.019],[366,-51.9131],[368,-52.2902]]
       }
       ring_03_l_ctrl.Scale.X {type=MovieSceneScriptingFloatChannel, default=1, const [1,18]}
       ring_03_l_ctrl.Scale.Y {type=MovieSceneScriptingFloatChannel, default=1, const [1,18]}
       ring_03_l_ctrl.Scale.Z {type=MovieSceneScriptingFloatChannel, default=1, const [1,18]}
       pinky_metacarpal_l_ctrl.Location.X {
        type=MovieSceneScriptingFloatChannel
        default=0.0001
        thinned_from=18
        keys [[0,0],[330,0],[340,-0.0008],[342,-0.0007],[344,0],[360,0],[362,-0.0008],[364,0],[366,-0.0001],[368,0.0001]]
       }
       pinky_metacarpal_l_ctrl.Location.Y {
        type=MovieSceneScriptingFloatChannel
        default=0
        thinned_from=18
        keys [[0,0],[330,0],[340,-0.0001],[360,0],[362,-0.0004],[366,0.0001],[368,0]]
       }
       pinky_metacarpal_l_ctrl.Location.Z {
        type=MovieSceneScriptingFloatChannel
        default=0.0001
        thinned_from=18
        keys [[0,0],[330,0],[340,-0.0006],[342,-0.0004],[344,0],[360,0],[362,-0.0011],[364,-0.0004],[368,0.0001]]
       }
       pinky_metacarpal_l_ctrl.Rotation.X {
        type=MovieSceneScriptingFloatChannel
        default=-7.8933
        thinned_from=18
        keys [[0,-5.9588],[330,-5.9588],[340,-1.5648],[342,-0.4372],[344,0],[346,-0.8209],[348,-2.7784],[350,-5.1148],[352,-7.0724],[354,-7.8933],[368,-7.8933]]
       }
       pinky_metacarpal_l_ctrl.Rotation.Y {
        type=MovieSceneScriptingFloatChannel
        default=-3.2809
        thinned_from=18
        keys [[0,-5.2205],[330,-5.2205],[340,-0.6504],[342,-0.1817],[344,0],[346,-0.3412],[348,-1.1549],[350,-2.126],[352,-2.9397],[354,-3.2809],[368,-3.2809]]
       }
       pinky_metacarpal_l_ctrl.Rotation.Z {
        type=MovieSceneScriptingFloatChannel
        default=-1.6954
        thinned_from=18
        keys [[0,-11.0008],[330,-11.0008],[340,-0.3361],[342,-0.0939],[344,0],[346,-0.1763],[348,-0.5968],[350,-1.0986],[352,-1.5191],[354,-1.6954],[368,-1.6954]]
       }
       pinky_metacarpal_l_ctrl.Scale.X {type=MovieSceneScriptingFloatChannel, default=1, const [1,18]}
       pinky_metacarpal_l_ctrl.Scale.Y {type=MovieSceneScriptingFloatChannel, default=1, const [1,18]}
       pinky_metacarpal_l_ctrl.Scale.Z {type=MovieSceneScriptingFloatChannel, default=1, const [1,18]}
       pinky_01_l_ctrl.Location.X {
        type=MovieSceneScriptingFloatChannel
        default=0.0002
        thinned_from=18
        keys [[0,0],[330,0],[340,-0.0021],[342,-0.0016],[344,0],[352,0.0002],[358,0.0002],[360,-0.0002],[362,-0.0023],[364,0],[366,-0.0001],[368,0.0002]]
       }
       pinky_01_l_ctrl.Location.Y {
        type=MovieSceneScriptingFloatChannel
        default=0.0003
        thinned_from=18
        keys [[0,0],[330,0],[340,-0.0006],[344,0],[358,0.0002],[360,0],[362,-0.0015],[364,-0.0005],[366,0.0001],[368,0.0003]]
       }
       pinky_01_l_ctrl.Location.Z {
        type=MovieSceneScriptingFloatChannel
        default=0
        thinned_from=18
        keys [[0,0],[330,0],[342,-0.0002],[344,0],[360,0],[362,0.0004],[364,0.0002],[366,-0.0002],[368,0]]
       }
       pinky_01_l_ctrl.Rotation.X {
        type=MovieSceneScriptingFloatChannel
        default=-2.9481
        thinned_from=18
        keys [[0,-11.7646],[330,-11.7646],[340,-1.236],[342,-0.3564],[344,0],[346,-0.1204],[348,-0.654],[350,-1.5205],[352,-2.1888],[354,-2.5384],[356,-2.6073],[358,-2.6293],[360,-2.6323],[362,-2.6274],[364,-2.9518],[366,-2.9892],[368,-2.9481]]
       }
       pinky_01_l_ctrl.Rotation.Y {
        type=MovieSceneScriptingFloatChannel
        default=6.2655
        thinned_from=18
        keys [[0,-1.0914],[330,-1.0914],[340,2.4141],[342,0.7519],[344,0],[346,0.1822],[348,1.4519],[350,3.0445],[352,4.5368],[354,5.0597],[356,5.0366],[358,5.0615],[360,5.1278],[362,5.1131],[364,6.112],[366,6.2776],[368,6.2655]]
       }
       pinky_01_l_ctrl.Rotation.Z {
        type=MovieSceneScriptingFloatChannel
        default=-65.63
        thinned_from=19
        keys [[0,-84.6646],[330,-84.6646],[340,-8.7948],[342,-1.9282],[344,0],[346,-8.5773],[348,-26.4162],[350,-47.4066],[352,-64.6973],[354,-72.1778],[356,-71.7073],[360,-71.2564],[362,-70.5475],[364,-65.6573],[366,-65.2931],[368,-65.63]]
       }
       pinky_01_l_ctrl.Scale.X {type=MovieSceneScriptingFloatChannel, default=1, const [1,18]}
       pinky_01_l_ctrl.Scale.Y {type=MovieSceneScriptingFloatChannel, default=1, const [1,18]}
       pinky_01_l_ctrl.Scale.Z {type=MovieSceneScriptingFloatChannel, default=1, const [1,18]}
       pinky_02_l_ctrl.Location.X {type=MovieSceneScriptingFloatChannel, default=-0.0001, thinned_from=18, keys [[0,0],[330,0],[340,-0.0005],[344,0],[368,-0.0001]]}
       pinky_02_l_ctrl.Location.Y {
        type=MovieSceneScriptingFloatChannel
        default=0.0001
        thinned_from=18
        keys [[0,0],[330,0],[340,-0.0004],[344,0],[358,0.0001],[360,0],[362,-0.0004],[364,0.0002],[368,0.0001]]
       }
       pinky_02_l_ctrl.Location.Z {
        type=MovieSceneScriptingFloatChannel
        default=0
        thinned_from=18
        keys [[0,0],[330,0],[340,0.0002],[360,0.0001],[362,0.0005],[364,0],[368,0]]
       }
       pinky_02_l_ctrl.Rotation.X {
        type=MovieSceneScriptingFloatChannel
        default=0.4931
        thinned_from=18
        keys [[0,0],[330,0],[340,1.469],[342,0.4434],[344,0],[346,-0.3045],[348,-0.2391],[350,-0.1603],[352,-0.0478],[354,-0.0398],[356,-0.0735],[358,-0.1174],[360,0.0227],[362,0.0067],[364,0.3614],[366,0.5313],[368,0.4931]]
       }
       pinky_02_l_ctrl.Rotation.Y {
        type=MovieSceneScriptingFloatChannel
        default=1.4446
        thinned_from=18
        keys [[0,0],[330,0],[340,2.1371],[342,0.7317],[344,0],[346,-0.5006],[348,-0.2527],[350,0.0406],[352,0.3817],[354,0.4423],[356,0.3699],[358,0.273],[360,0.5722],[362,0.5391],[364,1.2012],[366,1.5072],[368,1.4446]]
       }
       pinky_02_l_ctrl.Rotation.Z {
        type=MovieSceneScriptingFloatChannel
        default=-56.179
        thinned_from=18
        keys [[0,-63.0936],[330,-63.0936],[340,0.6734],[342,0.8544],[344,0],[346,-9.6303],[348,-24.2009],[350,-41.5964],[352,-55.6291],[354,-61.9801],[356,-62.3882],[358,-62.9314],[360,-61.242],[362,-61.4302],[364,-57.5954],[366,-55.804],[368,-56.179]]
       }
       pinky_02_l_ctrl.Scale.X {type=MovieSceneScriptingFloatChannel, default=1, const [1,18]}
       pinky_02_l_ctrl.Scale.Y {type=MovieSceneScriptingFloatChannel, default=1, const [1,18]}
       pinky_02_l_ctrl.Scale.Z {type=MovieSceneScriptingFloatChannel, default=1, const [1,18]}
       pinky_03_l_ctrl.Location.X {type=MovieSceneScriptingFloatChannel, default=0.0002, thinned_from=18, keys [[0,0],[360,0.0002],[362,-0.0004],[366,0.0002],[368,0.0002]]}
       pinky_03_l_ctrl.Location.Y {type=MovieSceneScriptingFloatChannel, default=-0.0002, thinned_from=18, keys [[0,0],[360,-0.0002],[362,0],[364,-0.0002],[368,-0.0002]]}
       pinky_03_l_ctrl.Location.Z {
        type=MovieSceneScriptingFloatChannel
        default=0
        thinned_from=18
        keys [[0,0],[330,0],[340,0.0001],[360,0],[362,0.0002],[364,-0.0001],[368,0]]
       }
       pinky_03_l_ctrl.Rotation.X {
        type=MovieSceneScriptingFloatChannel
        default=-0.6832
        thinned_from=18
        keys [[0,0],[330,0],[340,1.0729],[342,0.3542],[344,0],[346,-0.424],[348,-0.6206],[350,-0.8199],[352,-0.9689],[354,-1.0584],[356,-1.0803],[358,-1.1297],[360,-0.9749],[362,-1.0042],[364,-0.7676],[366,-0.6423],[368,-0.6832]]
       }
       pinky_03_l_ctrl.Rotation.Y {
        type=MovieSceneScriptingFloatChannel
        default=0.6729
        thinned_from=18
        keys [[0,0],[330,0],[340,0.8426],[342,0.2889],[344,0],[346,-0.1818],[348,-0.0387],[350,0.1563],[352,0.333],[354,0.3852],[356,0.3675],[358,0.3272],[360,0.4513],[362,0.4282],[364,0.6091],[366,0.7023],[368,0.6729]]
       }
       pinky_03_l_ctrl.Rotation.Z {
        type=MovieSceneScriptingFloatChannel
        default=-57.8398
        thinned_from=19
        keys [[0,-117.2954],[330,-117.2954],[340,-3.4334],[342,-0.5371],[344,0],[346,-8.497],[348,-23.0464],[350,-40.1671],[352,-54.3838],[354,-60.5439],[356,-60.7049],[358,-61.0694],[360,-59.9359],[362,-60.1493],[364,-58.4516],[366,-57.552],[368,-57.8398]]
       }
       pinky_03_l_ctrl.Scale.X {type=MovieSceneScriptingFloatChannel, default=1, const [1,18]}
       pinky_03_l_ctrl.Scale.Y {type=MovieSceneScriptingFloatChannel, default=1, const [1,18]}
       pinky_03_l_ctrl.Scale.Z {type=MovieSceneScriptingFloatChannel, default=1, const [1,18]}
       hand_l_ik_ctrl.Location.X {
        type=MovieSceneScriptingFloatChannel
        default=-31.4782
        thinned_from=140
        keys [[0,-402.7206],[92,-402.7206],[94,-401.7486],[96,-367.9136],[98,-324.4237],[100,-275.2537],[102,-223.6559],[104,-175.4097],[106,-135.1696],[108,-102.9202],[110,-86.0624],[112,-75.1831],[114,-63.3076],[116,-56.8112],[118,-49.9054],[120,-44.4041],[122,-40.4435],[124,-36.8141],[126,-34.1298],[128,-31.7475],[130,-29.4698],[134,-25.7424],[136,-23.9512],[138,-22.2214],[140,-20.6043],[142,-19.0919],[144,-17.6555],[148,-14.5858],[150,-13.1256],[152,-11.6061],[154,-10.5845],[156,-9.6389],[158,-7.949],[160,-6.3442],[162,-4.885],[164,-3.4928],[166,-2.1092],[168,-0.6858],[170,0.663],[172,1.8419],[174,2.8152],[176,3.6749],[178,4.4444],[180,5.1666],[182,5.8462],[184,6.4649],[186,7.0554],[188,7.6039],[190,8.1122],[192,8.565],[194,8.9317],[196,9.1558],[198,9.2128],[200,8.9859],[206,9.1001],[208,9.1072],[212,9.0373],[216,8.9017],[220,8.6312],[222,8.4719],[224,8.2407],[226,8.0393],[228,7.8602],[230,7.7147],[232,7.5283],[240,6.7252],[242,6.4897],[244,6.2713],[246,6.0095],[254,5.1216],[256,4.9315],[260,4.5854],[262,4.4403],[264,4.2797],[268,3.9915],[270,3.7998],[272,3.5951],[274,3.4654],[276,3.4163],[278,3.4176],[280,3.4259],[282,3.4044],[284,3.3898],[286,3.4269],[288,3.4343],[292,3.4775],[296,3.5673],[298,3.5981],[302,3.7305],[304,3.7904],[306,3.8977],[308,3.972],[314,4.3911],[320,4.6934],[322,4.7577],[324,4.8076],[326,4.7938],[328,4.6998],[330,4.8899],[332,5.161],[334,5.1848],[336,6.4977],[338,7.3572],[340,8.3805],[342,10.3778],[344,12.3205],[346,14.7526],[348,16.5477],[350,17.8109],[352,18.6371],[354,18.9349],[356,18.6277],[358,17.5026],[360,16.5764],[362,14.4542],[364,-4.5665],[366,-28.9151],[368,-31.4782]]
       }
       hand_l_ik_ctrl.Location.Y {
        type=MovieSceneScriptingFloatChannel
        default=-13.8448
        thinned_from=140
        keys [[0,8.2223],[92,8.2223],[94,8.6816],[96,8.2296],[98,7.9217],[100,7.6303],[102,7.721],[104,7.4413],[106,6.1171],[108,5.5029],[110,4.0908],[112,4.5904],[114,9.6143],[116,6.7659],[118,6.5337],[120,5.391],[122,3.6082],[124,1.9566],[126,0.1398],[128,-1.5845],[130,-3.1966],[132,-4.5971],[134,-5.6706],[136,-6.392],[138,-6.6354],[140,-6.6551],[142,-6.5],[144,-6.192],[146,-5.8511],[148,-5.5639],[150,-5.1132],[152,-4.8995],[154,-3.8443],[156,-2.981],[158,-3.7632],[160,-4.413],[162,-5.0252],[164,-5.7042],[166,-6.4729],[168,-7.3611],[170,-8.646],[172,-9.6164],[174,-10.4236],[176,-11.124],[178,-11.7316],[180,-12.2239],[182,-12.5775],[184,-12.8252],[186,-12.914],[188,-12.8884],[192,-12.708],[194,-12.5801],[196,-12.3862],[200,-12.1027],[204,-11.9776],[206,-11.999],[208,-12.1542],[210,-12.3655],[212,-12.4355],[214,-12.5312],[218,-13.0053],[220,-13.1592],[222,-13.2774],[224,-13.3427],[226,-13.7376],[228,-14.2095],[230,-14.9228],[234,-15.5522],[236,-15.7705],[238,-15.889],[244,-16.0394],[246,-16.1332],[252,-16.5652],[254,-16.6557],[258,-16.78],[262,-16.9734],[268,-17.0617],[272,-17.3248],[278,-17.3597],[284,-17.6593],[286,-17.6665],[288,-17.5812],[292,-17.3229],[296,-17.1461],[302,-17.1132],[306,-17.0458],[310,-16.8182],[312,-16.6596],[318,-16.3573],[322,-16.0713],[324,-15.8539],[326,-15.6927],[328,-15.6393],[330,-13.9436],[332,-10.3426],[334,-7.8309],[336,-11.7905],[338,-14.3373],[340,-15.017],[342,-14.8338],[344,-11.8018],[346,-9.5111],[348,-7.0482],[350,-5.1627],[352,-3.3964],[354,-2.059],[356,-1.6351],[358,-2.414],[360,-6.6147],[362,-9.5357],[364,-12.3439],[366,-17.0639],[368,-13.8448]]
       }
       hand_l_ik_ctrl.Location.Z {
        type=MovieSceneScriptingFloatChannel
        default=257.5527
        thinned_from=140
        keys [[0,-798.9283],[92,-798.9283],[94,-799.8132],[96,-734.2256],[98,-650.1113],[100,-554.8939],[102,-456.4626],[104,-360.7094],[106,-274.176],[108,-207.2949],[110,-164.4736],[112,-141.1844],[114,-133.2573],[116,-107.9674],[118,-96.8525],[120,-88.0251],[122,-79.6116],[124,-71.7444],[126,-62.8008],[128,-55.3243],[130,-49.6082],[132,-44.2462],[134,-39.8216],[136,-35.7719],[138,-31.9907],[140,-28.4761],[142,-25.2874],[144,-22.1759],[146,-19.1668],[148,-16.4226],[150,-13.9547],[152,-11.8263],[154,-10.2615],[158,-7.4788],[160,-6],[162,-4.8291],[164,-3.858],[166,-3.1218],[168,-2.6735],[170,-2.2654],[172,-1.8376],[174,-1.5568],[176,-1.3043],[178,-1.057],[180,-0.7927],[182,-0.4894],[184,-0.1421],[186,0.311],[190,1.3105],[192,1.7252],[194,1.7683],[196,1.4303],[198,0.7912],[200,-0.5808],[202,-0.9532],[204,-1.1743],[206,-1.3187],[208,-1.415],[210,-1.5144],[212,-1.4823],[214,-1.4209],[216,-1.4711],[218,-1.5952],[220,-1.6737],[222,-1.7155],[224,-1.736],[226,-1.703],[228,-1.5638],[230,-1.3943],[232,-1.2007],[234,-1.0039],[236,-0.7763],[238,-0.5327],[240,-0.3431],[242,-0.2071],[244,-0.0946],[246,-0.0652],[248,0.0434],[250,0.1354],[252,0.1978],[254,0.2552],[256,0.291],[258,0.3914],[260,0.4715],[262,0.6041],[264,0.7256],[266,0.8053],[268,0.7783],[270,0.6887],[272,0.4184],[274,0.1864],[276,0.0029],[278,-0.0964],[280,-0.1586],[282,-0.2777],[284,-0.395],[286,-0.3859],[288,-0.4357],[290,-0.482],[292,-0.4549],[294,-0.4718],[296,-0.4871],[298,-0.5477],[300,-0.6112],[304,-0.855],[306,-0.979],[308,-1.1738],[310,-1.2051],[312,-1.1684],[314,-1.1124],[316,-0.9924],[318,-0.7388],[320,-0.3511],[322,-0.0085],[324,0.45],[326,0.8731],[328,1.2826],[330,3.1191],[332,7.6008],[334,13.6842],[336,19.4359],[338,25.0802],[340,28.0716],[342,26.9116],[344,25.1532],[346,21.0816],[348,16.5919],[350,14.7967],[352,13.5134],[354,15.2132],[356,18.2026],[358,21.4175],[360,26.954],[362,33.1069],[364,65.0479],[366,126.7112],[368,257.5527]]
       }
       hand_l_ik_ctrl.Rotation.X {
        type=MovieSceneScriptingFloatChannel
        default=217.3267
        thinned_from=140
        keys [[0,127.0602],[92,127.0602],[96,118.1945],[98,114.1287],[100,111.0681],[102,108.5929],[104,108.4256],[106,110.2736],[108,111.1186],[110,113.4675],[112,114.4367],[114,109.6795],[116,113.6028],[118,110.7025],[122,102.0567],[124,98.3355],[126,97.0008],[128,96.1897],[130,95.0798],[132,94.813],[142,92.0738],[146,91.7032],[150,90.8046],[152,89.5679],[154,87.5716],[156,85.1284],[158,82.9377],[160,81.6672],[166,78.8769],[168,77.5985],[170,74.9549],[172,73.0055],[174,71.2325],[176,69.7692],[178,68.6654],[180,67.8005],[182,67.2177],[186,66.6084],[194,66.4584],[196,65.2687],[198,63.3344],[200,58.0774],[202,57.0415],[206,55.9643],[208,56.1264],[210,55.7214],[214,55.2904],[216,54.9089],[218,54.6421],[220,53.9946],[224,53.5924],[226,54.0487],[230,54.002],[232,53.768],[234,53.6525],[238,53.776],[242,53.717],[244,53.5953],[246,53.6487],[250,53.1438],[254,53.0398],[258,52.3776],[260,52.2865],[268,52.5314],[270,52.2382],[274,50.9201],[276,50.4613],[278,50.3167],[280,50.5122],[282,50.5401],[284,50.3636],[294,50.8206],[298,50.7846],[304,50.2049],[308,49.5722],[312,49.3776],[318,49.6105],[320,49.9057],[324,50.0618],[326,50.2291],[328,50.1072],[330,51.7075],[332,55.7982],[334,60.4471],[336,59.0368],[338,56.9516],[340,57.6844],[342,50.171],[344,45.9764],[346,37.1743],[348,38.8593],[350,33.8391],[352,37.8659],[354,40.5752],[356,44.539],[358,51.4159],[360,72.5778],[362,55.737],[364,132.5345],[366,215.5865],[368,217.3267]]
       }
       hand_l_ik_ctrl.Rotation.Y {
        type=MovieSceneScriptingFloatChannel
        default=57.6267
        thinned_from=140
        keys [[0,80.6339],[92,80.6339],[94,79.1892],[96,77.0234],[98,73.6085],[100,69.2495],[102,63.8145],[104,58.979],[106,55.6527],[108,51.2746],[110,48.6804],[112,42.6178],[114,26.0606],[116,23.418],[118,13.9704],[120,6.2889],[122,1.7491],[124,-0.7696],[126,-0.5367],[128,-0.5642],[130,-1.1206],[132,-0.7863],[134,-0.0404],[136,1.5985],[138,3.8247],[140,6.2363],[142,8.2932],[144,10.4267],[146,12.2181],[148,13.8244],[150,15.8749],[152,18.1604],[154,22.2049],[156,26.1111],[158,27.0458],[162,30.3419],[164,31.8856],[166,33.0676],[168,33.7007],[170,34.4209],[172,33.9124],[174,33.2022],[178,31.5885],[180,30.9205],[182,30.4013],[184,30.0492],[186,29.9307],[188,29.9789],[194,30.3769],[196,29.0962],[198,27.6223],[200,25.4748],[202,24.0886],[204,22.8865],[206,22.7279],[208,22.9138],[210,22.8808],[212,23.4682],[214,23.4113],[216,23.2369],[218,22.8413],[220,22.3543],[222,22.2807],[226,22.3126],[230,21.4951],[232,21.381],[234,21.4526],[236,21.6021],[240,21.71],[242,21.6953],[244,21.5615],[246,21.5859],[248,21.4949],[250,21.5764],[252,21.7307],[254,21.7565],[256,21.5712],[258,21.5769],[260,21.6691],[262,21.9464],[268,22.0669],[270,21.7144],[272,21.2698],[274,21.0095],[276,20.8585],[278,20.9489],[280,21.2444],[282,21.332],[284,21.1681],[286,21.1797],[288,21.1305],[290,21.1788],[292,21.179],[294,21.4426],[296,21.5234],[298,21.4409],[304,21.4833],[306,21.3128],[308,21.0545],[310,21.0373],[314,21.2445],[316,21.2213],[320,21.4355],[322,21.6397],[324,21.59],[326,21.5882],[328,21.2954],[330,23.0946],[332,27.2224],[334,31.6876],[336,43.6529],[338,57.431],[340,64.0596],[342,66.4414],[344,63.2147],[346,54.4733],[348,48.4659],[350,43.8308],[352,43.8236],[354,47.4204],[356,48.2887],[358,61.4544],[360,65.7359],[362,49.277],[364,60.736],[366,47.3428],[368,57.6267]]
       }
       hand_l_ik_ctrl.Rotation.Z {
        type=MovieSceneScriptingFloatChannel
        default=144.737
        thinned_from=140
        keys [[0,69.2689],[92,69.2689],[96,61.1269],[98,57.3173],[100,54.3649],[102,51.9966],[104,51.5294],[106,52.1905],[108,51.8226],[110,52.2403],[112,52.2941],[114,59.8221],[116,53.5606],[120,55.0953],[122,55.7792],[124,55.8589],[126,55.171],[128,54.1077],[132,51.5979],[134,50.5401],[136,49.7397],[138,49.5523],[140,49.7192],[142,50.1764],[144,50.9863],[146,51.6567],[150,52.5083],[152,52.3059],[154,53.4311],[156,53.4637],[158,50.8519],[160,48.9077],[162,47.1036],[164,45.0555],[166,42.6346],[168,39.6334],[170,35.4781],[172,32.4066],[174,29.594],[176,27.2557],[178,25.3169],[180,23.7125],[182,22.5173],[184,21.6654],[186,21.1875],[188,21.0297],[192,21.0649],[194,21.2853],[196,21.4395],[198,21.5365],[200,20.2925],[202,20.352],[204,20.777],[206,20.7309],[208,21.0249],[210,20.5281],[214,20.2774],[216,19.9244],[218,19.7942],[220,19.5291],[222,19.3267],[224,18.0357],[226,18.4927],[228,19.0371],[230,19.5015],[232,19.1736],[234,18.9086],[236,18.8555],[238,18.8881],[244,19.2254],[246,19.2892],[248,19.2538],[250,19.0546],[252,18.9658],[254,18.9265],[256,19.0158],[260,18.9537],[262,18.9526],[266,19.1986],[268,19.2872],[270,19.1354],[272,18.8021],[274,18.6157],[276,18.6516],[280,18.5801],[282,18.3448],[284,18.2646],[286,18.3041],[290,18.8544],[292,19.0604],[294,19.2217],[296,19.2185],[298,19.375],[300,19.3407],[304,19.1516],[306,19.0994],[308,19.1645],[314,19.9127],[318,20.457],[320,20.8052],[322,20.8456],[324,21.0206],[326,21.2462],[328,21.1938],[330,24.6201],[332,33.0322],[334,39.9489],[336,31.8413],[338,26.139],[340,25.7377],[342,16.1321],[344,14.5691],[346,11.9604],[348,11.89],[350,12.8225],[352,11.2451],[354,14.5342],[356,25.0336],[358,23.8278],[360,42.6114],[362,33.3313],[364,81.5351],[366,140.9689],[368,144.737]]
       }
       hand_l_ik_ctrl.Scale.X {type=MovieSceneScriptingFloatChannel, default=1, const [1,140]}
       hand_l_ik_ctrl.Scale.Y {type=MovieSceneScriptingFloatChannel, default=1, const [1,140]}
       hand_l_ik_ctrl.Scale.Z {type=MovieSceneScriptingFloatChannel, default=1, const [1,140]}
       Softness {type=MovieSceneScriptingFloatChannel, default=0, const [0,140]}
       hand_l_ikorient_ctrl.Rotation.Roll {
        type=MovieSceneScriptingFloatChannel
        default=0
        thinned_from=140
        keys [[0,0],[344,0],[346,14.6393],[348,0],[350,16.3146],[352,0],[354,0],[356,17.2037],[358,0],[360,16.7825],[362,0],[368,0]]
       }
       hand_l_ikorient_ctrl.Rotation.Pitch {
        type=MovieSceneScriptingFloatChannel
        default=0
        thinned_from=140
        keys [[0,0],[344,0],[346,23.4231],[348,0],[350,23.9173],[352,0],[354,0],[356,24.2868],[358,0],[360,26.1318],[362,0],[368,0]]
       }
       hand_l_ikorient_ctrl.Rotation.Yaw {
        type=MovieSceneScriptingFloatChannel
        default=0
        thinned_from=140
        keys [[0,0],[344,0],[346,-14.7362],[348,0],[350,-11.0489],[352,0],[354,0],[356,-8.9979],[358,0],[360,-12.2491],[362,0],[368,0]]
       }
       arm_l_pv_ik_ctrl.Location.X {
        type=MovieSceneScriptingFloatChannel
        default=-11.6401
        thinned_from=140
        keys [[0,25.1076],[92,25.1076],[96,25.6082],[98,25.7183],[100,25.7713],[102,25.7079],[108,24.8277],[110,24.501],[112,24.0748],[114,25.1053],[116,24.1372],[118,25.5191],[120,27.4629],[122,29.2945],[124,30.5975],[126,31.3873],[130,32.6794],[132,33.1071],[134,33.4367],[136,33.6104],[138,33.6162],[140,33.5448],[144,33.2062],[148,32.7115],[152,31.9809],[154,31.2083],[156,30.3693],[162,30.1008],[166,29.8574],[168,29.7955],[172,30.1151],[174,30.1228],[186,29.8398],[192,29.8306],[194,29.9546],[196,29.904],[198,29.7291],[200,29.3868],[208,29.3888],[212,29.3151],[214,29.3297],[220,29.5338],[224,29.4977],[226,29.5391],[228,29.6586],[230,29.9293],[234,30.1533],[238,30.2295],[246,30.1448],[278,30.04],[286,30.2029],[294,29.9054],[304,29.8759],[318,29.9509],[322,29.8242],[326,29.5672],[332,28.9925],[334,28.9435],[336,29.2623],[338,29.7878],[340,29.9843],[342,28.7248],[344,26.4419],[346,23.4365],[348,20.6121],[350,19.091],[352,18.291],[354,18.6737],[356,19.5545],[358,22.0521],[360,27.3547],[362,34.2392],[364,37.3384],[366,20.4694],[368,-11.6401]]
       }
       arm_l_pv_ik_ctrl.Location.Y {
        type=MovieSceneScriptingFloatChannel
        default=239.46
        thinned_from=140
        keys [[0,-878.7301],[94,-879.2521],[96,-805.5739],[98,-710.8877],[100,-603.5867],[102,-492.106],[104,-384.3085],[106,-288.3542],[108,-213.3997],[110,-167.0272],[112,-140.1789],[114,-124.9798],[116,-98.5743],[118,-84.1936],[120,-73.2912],[122,-64.3977],[124,-56.2236],[126,-47.8741],[128,-40.7019],[130,-34.9103],[132,-29.8002],[134,-25.7131],[136,-22.3382],[138,-19.1797],[140,-16.247],[142,-13.5785],[144,-10.9543],[146,-8.2737],[148,-5.7848],[150,-3.525],[152,-1.5827],[154,-0.4788],[156,0.1144],[158,1.1979],[160,2.2764],[162,3.0154],[164,3.5999],[166,4.1374],[168,4.6646],[170,5.2655],[172,6.1011],[174,6.7392],[176,7.3211],[178,7.8728],[180,8.4084],[184,9.5493],[190,11.4761],[192,12.0136],[194,12.6071],[196,12.9673],[198,13.0163],[200,12.7196],[202,12.7146],[204,12.6295],[206,12.4044],[208,12.0933],[210,11.727],[212,11.3959],[214,11.2698],[216,11.0451],[222,10.5417],[224,10.3372],[226,10.1556],[228,10.0809],[230,10.0966],[232,10.1333],[236,10.1767],[240,10.2575],[242,10.247],[244,10.2151],[246,10.1512],[248,10.1142],[250,10.0423],[254,9.8693],[256,9.8016],[258,9.7693],[260,9.7033],[262,9.6863],[266,9.6111],[268,9.5337],[270,9.3901],[272,9.1679],[278,8.6425],[282,8.3584],[284,8.2431],[286,8.2374],[290,8.1805],[296,8.0186],[298,7.9292],[300,7.8578],[304,7.629],[308,7.4734],[310,7.5105],[312,7.5864],[314,7.6888],[316,7.8461],[318,8.0922],[320,8.438],[322,8.7427],[324,9.1812],[326,9.5806],[328,9.8675],[330,11.301],[332,14.6333],[334,18.4758],[336,19.8065],[338,22.005],[340,23.4989],[342,22.1113],[344,21.5419],[346,18.9872],[348,16.5554],[350,16.0284],[352,16.0804],[354,17.5235],[356,19.3254],[358,21.4015],[360,25.5022],[362,32.3014],[364,63.8607],[366,118.9608],[368,239.46]]
       }
       arm_l_pv_ik_ctrl.Location.Z {
        type=MovieSceneScriptingFloatChannel
        default=121.9039
        thinned_from=140
        keys [[0,-1.634],[92,-1.634],[94,-2.6599],[96,-3.7325],[98,-4.9847],[100,-6.1399],[102,-7.7857],[104,-7.77],[106,-4.7961],[108,-2.9514],[110,1.8047],[112,3.6449],[114,-2.4457],[116,4.7296],[118,4.9694],[120,5.5017],[122,7.0756],[124,8.5965],[126,10.9753],[128,13.1123],[130,14.7217],[132,16.162],[134,16.9177],[136,17.1111],[138,16.8482],[140,16.3841],[142,15.7663],[144,14.9081],[146,13.9144],[148,12.9619],[152,10.8785],[154,9.5599],[156,8.4947],[158,8.0669],[160,7.5838],[162,7.1827],[164,6.9135],[166,6.6965],[168,6.4455],[170,6.4193],[174,6.2711],[178,6.0486],[180,5.8773],[182,5.658],[184,5.4269],[186,5.1282],[188,4.8124],[190,4.4796],[192,4.1763],[194,3.9489],[196,3.6999],[198,3.5386],[200,3.5308],[202,3.4526],[204,3.3554],[206,3.2725],[208,3.252],[210,3.3133],[212,3.3553],[214,3.461],[216,3.5812],[218,3.742],[220,3.8767],[222,4.0193],[224,4.2559],[226,4.6158],[228,5.0178],[230,5.508],[234,6.218],[236,6.5296],[238,6.8117],[240,7.0484],[242,7.3211],[244,7.5791],[246,7.9032],[250,8.5185],[254,9.0574],[260,9.6849],[264,10.0685],[266,10.2345],[268,10.4303],[270,10.7141],[272,10.9741],[274,11.077],[276,11.0579],[278,11.0124],[280,11.0154],[284,11.1671],[286,11.2073],[288,11.2103],[290,11.1713],[292,11.1051],[294,11.0157],[296,10.9545],[304,10.7647],[308,10.6249],[312,10.3278],[314,10.2006],[316,10.1353],[320,10.1193],[322,10.1312],[324,10.1884],[326,10.3585],[328,10.7102],[330,10.1529],[332,8.4576],[334,6.9028],[336,6.7216],[338,5.2147],[340,3.0995],[342,0.8582],[344,-1.9617],[346,-4.2242],[348,-5.7207],[350,-6.8685],[352,-7.3752],[354,-8.2782],[356,-8.924],[358,-8.6932],[360,-7.8475],[362,-5.976],[364,17.3556],[366,61.6208],[368,121.9039]]
       }
       arm_l_pv_ik_ctrl.Rotation.X {
        type=MovieSceneScriptingFloatChannel
        default=-83.0591
        thinned_from=140
        keys [[0,-30.7334],[92,-30.7334],[94,-29.0242],[96,-26.3764],[98,-22.1772],[100,-17.282],[102,-12.0726],[104,-7.9877],[106,-4.7792],[108,-1.0216],[110,1.8767],[112,4.9629],[114,4.7268],[116,8.8665],[118,10.0115],[120,11.3876],[122,12.8953],[124,14.4917],[126,15.2508],[128,16.1909],[130,17.3657],[132,17.9755],[134,18.1716],[136,17.727],[138,16.8608],[140,15.8128],[142,14.7116],[144,13.2082],[146,11.6616],[148,10.3873],[150,9.1599],[152,8.0568],[154,6.9732],[156,6.3702],[158,6.3948],[160,6.3885],[162,6.6432],[164,7.1397],[170,9.0181],[172,9.5246],[174,9.9796],[176,10.3106],[178,10.5945],[180,10.8023],[182,10.941],[184,11.0167],[186,10.952],[188,10.8437],[190,10.6869],[192,10.5636],[196,10.8271],[198,11.0836],[200,11.769],[202,12.031],[204,12.111],[206,12.133],[208,12.0833],[214,12.0828],[218,12.1423],[220,12.1335],[222,12.0997],[228,12.3476],[230,12.4638],[232,12.4839],[234,12.4771],[240,12.3513],[242,12.3497],[244,12.3786],[246,12.5059],[254,12.7407],[256,12.7722],[260,12.7664],[262,12.7339],[266,12.7886],[268,12.9456],[270,13.1966],[272,13.5146],[274,13.6793],[276,13.7098],[278,13.7019],[280,13.7231],[282,13.8545],[284,14.0598],[288,14.3544],[290,14.4481],[294,14.4223],[296,14.4603],[300,14.6203],[304,14.8875],[306,15.081],[308,15.3293],[310,15.3968],[318,15.4713],[324,15.3911],[326,15.4875],[328,15.7741],[330,14.8381],[332,11.7459],[334,7.7107],[336,4.5518],[338,0.4664],[340,-2.087],[342,-1.5544],[344,-1.0322],[346,1.7447],[348,4.3904],[350,6.1152],[352,7.464],[354,6.7816],[356,5.2691],[358,3.0891],[360,1.5318],[362,0.8248],[364,-20.2851],[366,-99.1465],[368,-83.0591]]
       }
       arm_l_pv_ik_ctrl.Rotation.Y {
        type=MovieSceneScriptingFloatChannel
        default=82.1786
        keys [[0,64.4758],[92,64.4758],[94,63.3136],[96,61.7658],[98,59.5147],[100,56.6571],[102,52.8709],[104,49.8644],[106,48.6574],[108,46.3371],[110,45.9485],[112,42.4207],[114,22.9673],[116,29.0344],[118,21.0917],[120,13.4665],[122,7.9313],[124,3.9718],[126,3.3045],[128,2.792],[130,1.7562],[132,1.8237],[134,1.9681],[136,2.6809],[138,3.3649],[140,3.7983],[142,3.8027],[144,3.7408],[146,3.5662],[148,3.2394],[150,2.6077],[152,1.9238],[154,0.1068],[156,-1.709],[158,-1.8972],[160,-1.9736],[162,-2.1917],[164,-2.3237],[166,-2.3392],[168,-2.149],[170,-1.4394],[172,-0.8551],[174,-0.4837],[176,-0.162],[178,0.1669],[180,0.4681],[182,0.7225],[184,0.9555],[186,1.0795],[188,1.1464],[190,1.1557],[192,1.1332],[194,0.8122],[196,0.1356],[198,-0.6737],[200,-2.0318],[202,-2.4725],[204,-2.7333],[206,-2.7953],[208,-2.6761],[210,-2.5026],[212,-2.3535],[214,-2.2664],[216,-2.1428],[218,-2.1632],[220,-2.2266],[222,-2.3024],[224,-2.396],[226,-2.0988],[228,-1.6836],[230,-1.0564],[232,-0.8308],[234,-0.6125],[236,-0.4806],[238,-0.4488],[240,-0.5272],[242,-0.5857],[244,-0.6913],[246,-0.7948],[248,-0.7831],[250,-0.7316],[252,-0.7102],[254,-0.7275],[256,-0.8075],[258,-0.81],[260,-0.7609],[262,-0.7031],[264,-0.6869],[266,-0.698],[268,-0.7452],[270,-0.6959],[272,-0.7496],[274,-0.8887],[276,-1.0215],[278,-1.0419],[280,-0.9165],[282,-0.7852],[284,-0.6984],[286,-0.6373],[288,-0.7111],[290,-0.8321],[292,-0.8917],[294,-0.937],[296,-0.9523],[298,-0.9392],[300,-0.8894],[302,-0.849],[304,-0.854],[306,-0.8786],[308,-1.0425],[310,-1.1608],[312,-1.2902],[314,-1.3585],[316,-1.4108],[318,-1.4002],[320,-1.3766],[322,-1.4096],[324,-1.4743],[326,-1.4705],[328,-1.3042],[330,-2.3012],[332,-3.7586],[334,-2.5594],[336,5.1021],[338,11.6328],[340,15.7216],[342,16.883],[344,13.9091],[346,10.8939],[348,8.604],[350,7.4471],[352,7.3235],[354,7.028],[356,7.3432],[358,9.6866],[360,14.3402],[362,21.2419],[364,59.1956],[366,69.9161],[368,82.1786]]
       }
       arm_l_pv_ik_ctrl.Rotation.Z {
        type=MovieSceneScriptingFloatChannel
        default=12.6697
        thinned_from=140
        keys [[0,52.8375],[92,52.8375],[94,54.2037],[96,56.5367],[98,60.5146],[100,65.1974],[102,70.2307],[104,74.4015],[106,77.7446],[108,81.4685],[110,84.3946],[112,87.7532],[114,84.7586],[116,91.185],[118,89.6404],[120,86.3128],[122,82.1156],[124,78.6531],[126,76.3643],[130,73.1151],[132,71.8734],[134,70.3709],[136,68.5949],[142,63.7932],[144,62.3551],[150,58.51],[152,57.0269],[154,55.1279],[156,52.5905],[160,47.9003],[164,42.8993],[166,40.5261],[168,38.3991],[170,36.4653],[172,35.3252],[176,32.9447],[178,31.9228],[180,31.0271],[182,30.3473],[184,29.8662],[186,29.5721],[188,29.4729],[192,29.433],[194,30.0372],[196,30.4947],[198,30.6419],[200,30.683],[202,30.8381],[204,30.8158],[206,30.5139],[208,29.926],[210,29.1813],[212,28.6082],[214,28.3591],[216,27.9729],[218,27.7122],[220,27.509],[224,26.9899],[226,26.5105],[228,26.1419],[230,25.8264],[232,25.7455],[236,25.4394],[242,25.2531],[248,24.912],[252,24.5473],[256,24.2824],[258,24.1977],[260,24.0452],[264,23.8686],[268,23.7703],[272,23.5029],[274,23.4539],[276,23.342],[280,22.9968],[282,22.874],[290,22.709],[294,22.4556],[296,22.3959],[300,22.2153],[302,22.0759],[308,21.9587],[312,22.0858],[318,22.4246],[322,22.678],[324,22.8895],[326,23.0491],[328,22.9445],[330,24.804],[332,29.2321],[334,32.2653],[336,25.6465],[338,20.6283],[340,17.2791],[342,13.2042],[344,12.667],[346,10.6191],[348,8.3364],[350,8.7365],[352,9.3917],[354,11.567],[356,13.8024],[358,15.9615],[360,21.9639],[362,31.6492],[364,31.438],[366,-19.8881],[368,12.6697]]
       }
       arm_l_pv_ik_ctrl.Scale.X {type=MovieSceneScriptingFloatChannel, default=1, const [1,140]}
       arm_l_pv_ik_ctrl.Scale.Y {type=MovieSceneScriptingFloatChannel, default=1, const [1,140]}
       arm_l_pv_ik_ctrl.Scale.Z {type=MovieSceneScriptingFloatChannel, default=1, const [1,140]}
       upperarm_r_fk_ctrl.Location.X {type=MovieSceneScriptingFloatChannel, default=0.0001, thinned_from=93, keys [[0,0],[116,-0.0001],[330,0],[332,-0.0001],[368,0.0001]]}
       upperarm_r_fk_ctrl.Location.Y {type=MovieSceneScriptingFloatChannel, default=0, thinned_from=93, keys [[0,0],[342,-0.0001],[368,0]]}
       upperarm_r_fk_ctrl.Location.Z {type=MovieSceneScriptingFloatChannel, default=-0.0001, thinned_from=93, keys [[0,0],[116,-0.0002],[364,-0.0001],[366,-0.0002],[368,-0.0001]]}
       upperarm_r_fk_ctrl.Rotation.X {
        type=MovieSceneScriptingFloatChannel
        default=74.4624
        thinned_from=93
        keys [[0,58.469],[92,58.469],[94,57.4702],[98,54.9209],[102,51.9984],[110,45.5826],[112,43.5526],[114,40.9215],[116,37.9862],[154,15.7313],[174,13.323],[176,12.556],[178,12.1282],[182,11.356],[214,11.3617],[216,11.3449],[218,11.2473],[220,10.9027],[222,9.944],[246,-18.2819],[248,-19.3393],[250,-19.5768],[252,-19.9649],[254,-20.1439],[256,-20.0149],[258,-19.8107],[260,-19.7015],[262,-19.5469],[264,-19.3405],[266,-19.0158],[268,-19.2334],[270,-19.3573],[272,-19.3774],[276,-19.3089],[278,-19.3201],[280,-19.3791],[282,-19.4968],[284,-19.8705],[286,-20.0492],[288,-20.0567],[294,-19.9362],[296,-19.9257],[300,-20.064],[302,-20.1891],[306,-20.3409],[308,-20.5043],[310,-20.5334],[314,-20.4391],[316,-20.4444],[320,-20.249],[322,-20.1967],[324,-20.017],[326,-19.777],[328,-18.5631],[330,-16.0403],[332,-12.582],[334,-8.1577],[336,-3.5413],[338,1.3049],[340,4.8562],[342,7.8023],[344,7.1047],[346,4.1505],[348,1.4149],[350,0.5833],[352,0.1883],[354,1.8572],[356,4.5028],[358,5.6164],[360,7.7356],[362,6.1546],[364,46.0372],[366,64.5163],[368,74.4624]]
       }
       upperarm_r_fk_ctrl.Rotation.Y {
        type=MovieSceneScriptingFloatChannel
        default=35.1669
        thinned_from=93
        keys [[0,36.4403],[92,36.4403],[96,36.0139],[98,35.6147],[100,35.0749],[102,34.4066],[104,33.6217],[106,32.7321],[108,31.7499],[110,30.6869],[112,28.8014],[114,25.6472],[116,21.6801],[154,5.8115],[174,6.0601],[176,6.0019],[178,6.2412],[180,6.468],[182,6.5919],[214,6.5973],[218,6.6283],[222,6.6819],[246,-8.5397],[248,-8.8098],[250,-9.1806],[252,-9.0888],[254,-8.7963],[256,-8.5321],[258,-8.3629],[260,-8.3094],[262,-8.281],[264,-8.2908],[266,-8.0393],[268,-7.2533],[272,-5.743],[274,-5.3261],[276,-5.2037],[278,-4.9039],[280,-4.51],[282,-4.3843],[284,-4.5562],[286,-4.8459],[288,-5.1138],[290,-5.3964],[292,-5.5136],[296,-5.4673],[298,-5.368],[300,-5.2917],[302,-5.2039],[304,-5.1704],[306,-5.2246],[308,-5.3541],[310,-5.4435],[314,-5.5559],[316,-5.6256],[318,-5.6797],[322,-5.581],[324,-5.3847],[326,-5.23],[328,-4.547],[330,-2.9089],[332,-0.3646],[334,3.4814],[336,9.0024],[338,13.3157],[340,15.6801],[342,14.8736],[344,14.2865],[346,12.0504],[348,10.3328],[352,10.6733],[354,9.6866],[356,7.9232],[358,5.6986],[360,2.057],[362,-0.1761],[364,32.2181],[366,40.5665],[368,35.1669]]
       }
       upperarm_r_fk_ctrl.Rotation.Z {
        type=MovieSceneScriptingFloatChannel
        default=170.8017
        thinned_from=93
        keys [[0,199.6101],[92,199.6101],[94,197.3778],[96,193.8628],[98,188.7794],[100,182.3279],[102,174.7086],[104,166.1217],[106,156.7679],[108,146.8472],[110,136.5601],[112,124.692],[114,110.6527],[116,95.5791],[154,2.5611],[174,-0.5784],[176,-1.6675],[178,-2.3748],[180,-2.9505],[182,-3.4358],[212,-3.4256],[214,-3.3914],[216,-3.28],[218,-2.9605],[220,-2.1305],[222,-0.168],[246,64.4358],[248,68.0134],[250,70.2584],[252,71.7635],[254,72.6121],[256,73.2467],[258,73.596],[264,73.8724],[276,76.0593],[280,76.4591],[282,76.38],[284,75.8508],[286,75.486],[290,75.16],[300,75.1067],[306,74.7114],[312,74.2205],[328,74.0005],[330,73.7843],[332,72.6852],[334,69.0659],[336,57.9853],[338,45.001],[340,31.2777],[342,19.5229],[344,13.9727],[346,11.3705],[348,10.1874],[350,11.9855],[352,14.8746],[354,16.7333],[356,18.9299],[358,21.7548],[360,25.5095],[362,30.929],[364,129.1751],[366,160.0236],[368,170.8017]]
       }
       upperarm_r_fk_ctrl.Scale.X {type=MovieSceneScriptingFloatChannel, default=1, const [1,93]}
       upperarm_r_fk_ctrl.Scale.Y {type=MovieSceneScriptingFloatChannel, default=1, const [1,93]}
       upperarm_r_fk_ctrl.Scale.Z {type=MovieSceneScriptingFloatChannel, default=1, const [1,93]}
       lowerarm_r_fk_ctrl.Location.X {
        type=MovieSceneScriptingFloatChannel
        default=-0.0002
        thinned_from=93
        keys [[0,-0.0001],[94,-0.0002],[112,-0.002],[114,-0.0024],[126,-0.0024],[154,-0.0001],[220,-0.0002],[222,-0.0005],[252,-0.0002],[254,0],[256,-0.0001],[258,0],[260,-0.0002],[262,-0.0001],[272,0.0001],[276,-0.0002],[278,-0.0002],[280,0.0001],[282,0],[284,0.0001],[286,-0.0003],[292,0],[296,-0.0001],[302,0.0001],[304,-0.0002],[308,0],[310,-0.0002],[312,0.0001],[314,0],[316,0.0002],[318,-0.0002],[320,0.0002],[322,0.0002],[324,-0.0001],[328,-0.0003],[330,0],[332,-0.0036],[334,-0.0001],[336,-0.0044],[338,-0.0057],[340,-0.0092],[342,-0.0079],[344,-0.0012],[346,-0.0005],[348,-0.0002],[362,0],[364,-0.0002],[366,-0.0022],[368,-0.0002]]
       }
       lowerarm_r_fk_ctrl.Location.Y {
        type=MovieSceneScriptingFloatChannel
        default=0.0001
        thinned_from=93
        keys [[0,0.0001],[92,0.0001],[110,0.0019],[114,0.0028],[126,0.0028],[154,0.0001],[222,0.0001],[246,-0.0001],[248,0.0001],[250,-0.0001],[254,0.0003],[256,0],[258,0.0002],[260,-0.0002],[262,0],[266,-0.0001],[268,0.0002],[270,0.0002],[272,0.0004],[276,-0.0001],[278,-0.0002],[280,0.0003],[282,0.0002],[284,0.0005],[286,-0.0003],[290,-0.0001],[292,0.0003],[296,-0.0002],[298,0.0001],[302,0.0004],[304,-0.0002],[308,0.0002],[310,-0.0001],[312,0.0003],[314,0.0002],[316,0.0006],[318,-0.0002],[320,0.0005],[322,0.0005],[324,-0.0001],[328,-0.0002],[330,0.0002],[332,0.003],[334,0],[336,0.0026],[338,0.0039],[340,0.004],[342,0.002],[344,0.0002],[364,0.0001],[366,0.0011],[368,0.0001]]
       }
       lowerarm_r_fk_ctrl.Location.Z {
        type=MovieSceneScriptingFloatChannel
        default=0
        thinned_from=93
        keys [[0,-0.0001],[92,-0.0001],[110,-0.0024],[114,-0.0035],[126,-0.0035],[154,0],[220,-0.0001],[222,-0.0005],[246,0.0003],[248,0.0001],[250,0.0002],[254,-0.0003],[256,0.0001],[258,-0.0002],[260,0.0002],[262,0],[266,0.0001],[268,-0.0001],[272,-0.0003],[276,0.0002],[278,0.0002],[280,-0.0003],[282,-0.0001],[284,-0.0004],[286,0.0003],[290,0.0001],[292,-0.0002],[296,0.0002],[298,-0.0001],[302,-0.0003],[304,0.0002],[308,-0.0002],[310,0.0001],[312,-0.0002],[314,-0.0002],[316,-0.0005],[318,0.0002],[320,-0.0004],[322,-0.0004],[324,0.0001],[328,0.0002],[330,-0.0002],[332,0.0015],[334,0.0001],[336,0.0037],[338,0.0056],[340,0.0077],[342,0.0052],[344,0.0004],[364,0],[366,0.0004],[368,0]]
       }
       lowerarm_r_fk_ctrl.Rotation.X {
        type=MovieSceneScriptingFloatChannel
        default=2.1444
        thinned_from=93
        keys [[0,1.693],[92,1.693],[94,1.7303],[96,1.8061],[98,1.9182],[100,2.0646],[102,2.2431],[104,2.4515],[106,2.6879],[108,2.95],[110,3.2357],[112,3.5592],[114,3.9269],[126,6.9072],[154,12.5981],[174,13.1457],[178,13.4998],[180,13.6424],[182,13.7533],[214,13.7448],[216,13.7194],[218,13.6401],[220,13.426],[222,12.9219],[246,1.8964],[248,0.5156],[250,-0.3145],[252,-1.0341],[254,-1.5297],[256,-1.9712],[258,-2.2743],[260,-2.4215],[262,-2.6161],[264,-2.7854],[266,-3.1225],[268,-3.3902],[270,-3.6343],[272,-3.5183],[274,-3.2913],[276,-2.9702],[278,-3.0727],[282,-3.7924],[284,-4.0996],[286,-4.2389],[290,-4.2267],[296,-4.2201],[300,-4.2299],[304,-4.2514],[308,-4.2974],[316,-4.3279],[322,-4.3463],[328,-4.4184],[330,-4.2894],[332,-3.957],[334,-3.6439],[336,-3.936],[338,-3.9009],[340,-3.0236],[342,-1.5143],[344,-0.2782],[346,0.6627],[348,1.0956],[352,1.0982],[354,1.039],[356,0.9912],[358,0.9656],[360,0.9887],[362,1.0368],[364,1.7836],[366,2.5694],[368,2.1444]]
       }
       lowerarm_r_fk_ctrl.Rotation.Y {
        type=MovieSceneScriptingFloatChannel
        default=-0.3656
        thinned_from=93
        keys [[0,0.6378],[92,0.6378],[94,0.3882],[96,0.0766],[98,-0.2875],[100,-0.6947],[102,-1.1356],[104,-1.6007],[106,-2.0806],[110,-3.0469],[112,-3.7726],[114,-4.884],[126,-6.703],[154,-9.9984],[174,-9.625],[178,-9.3811],[180,-9.2803],[182,-9.2032],[214,-9.2096],[216,-9.2285],[218,-9.2884],[220,-9.4495],[246,-14.354],[248,-15.7505],[250,-16.4373],[252,-16.9628],[254,-17.2734],[256,-17.5179],[258,-17.6686],[266,-17.95],[268,-17.8485],[270,-17.6286],[272,-17.1068],[274,-16.3182],[276,-15.1682],[278,-14.2279],[280,-13.6118],[282,-13.1749],[284,-12.9769],[296,-12.9245],[304,-12.9969],[308,-13.0834],[322,-13.2045],[328,-13.3373],[330,-13.4295],[332,-13.6865],[334,-14.2934],[336,-15.5641],[338,-16.2988],[340,-16.3665],[342,-15.9616],[346,-15.2309],[348,-15.0106],[352,-15.0094],[356,-15.1185],[358,-15.1454],[362,0.4283],[364,-0.198],[366,-2.7717],[368,-0.3656]]
       }
       lowerarm_r_fk_ctrl.Rotation.Z {
        type=MovieSceneScriptingFloatChannel
        default=-14.8692
        thinned_from=93
        keys [[0,-18.5093],[92,-18.5093],[94,-17.1613],[96,-15.5061],[98,-13.5941],[100,-11.4762],[102,-9.2029],[104,-6.8251],[108,-1.9583],[110,0.4292],[112,4.1604],[114,10.0338],[126,34.0697],[154,24.0248],[174,21.4051],[178,19.683],[180,18.9746],[182,18.4326],[212,18.4408],[214,18.4776],[216,18.6123],[218,19.0402],[220,20.1939],[222,22.8728],[246,66.1793],[248,76.6707],[250,82.4107],[252,87.3807],[254,90.7711],[256,93.7034],[258,95.669],[260,96.6237],[264,98.9384],[266,100.7493],[268,101.6263],[270,101.9357],[272,99.5078],[274,95.7286],[276,90.3559],[278,87.628],[280,87.1812],[282,87.3374],[284,88.0958],[286,88.6702],[296,88.4527],[304,88.9026],[308,89.4493],[322,90.1216],[328,90.92],[330,90.6352],[332,89.9854],[334,90.7844],[336,97.0916],[338,97.415],[340,88.4983],[344,55.8338],[346,40.6395],[348,31.2351],[350,28.7572],[352,27.6308],[354,30.1813],[356,33.785],[358,38.5416],[360,45.0559],[362,53.9517],[364,15.1408],[366,-2.2115],[368,-14.8692]]
       }
       lowerarm_r_fk_ctrl.Scale.X {type=MovieSceneScriptingFloatChannel, default=1, const [1,93]}
       lowerarm_r_fk_ctrl.Scale.Y {type=MovieSceneScriptingFloatChannel, default=1, const [1,93]}
       lowerarm_r_fk_ctrl.Scale.Z {type=MovieSceneScriptingFloatChannel, default=1, const [1,93]}
       hand_r_fk_ctrl.Location.X {
        type=MovieSceneScriptingFloatChannel
        default=-0.0004
        thinned_from=93
        keys [[0,-0.0005],[92,-0.0005],[98,-0.0019],[110,-0.0067],[112,-0.0083],[114,-0.0112],[116,-0.0148],[154,-0.0002],[220,-0.0002],[222,-0.0011],[246,-0.0039],[250,-0.0024],[252,-0.0022],[254,-0.001],[256,-0.0009],[258,-0.0003],[260,-0.0004],[262,-0.0001],[264,-0.0003],[266,-0.0001],[268,-0.0007],[270,-0.0002],[272,-0.0084],[274,-0.0001],[276,-0.0073],[278,-0.0004],[280,-0.002],[282,-0.0002],[284,-0.0004],[286,-0.0001],[330,-0.0001],[332,-0.0014],[334,-0.0001],[336,-0.0024],[338,-0.005],[340,-0.0095],[342,-0.0085],[344,-0.001],[346,0],[348,-0.0002],[350,0],[354,-0.0002],[356,0],[358,-0.0001],[362,0],[364,-0.0002],[366,-0.0085],[368,-0.0004]]
       }
       hand_r_fk_ctrl.Location.Y {
        type=MovieSceneScriptingFloatChannel
        default=0
        thinned_from=93
        keys [[0,0.0001],[94,0.0002],[110,0.0017],[112,0.0022],[116,0.0039],[154,0],[222,-0.0001],[246,0.0011],[252,0.0006],[254,0.0003],[268,-0.0001],[270,-0.0001],[272,0.0007],[274,0],[276,0.0007],[278,0],[334,0],[336,0.0002],[338,0.0018],[340,0.0053],[342,0.0059],[344,0.0008],[346,-0.431],[348,0.0001],[350,-0.431],[352,0],[354,0],[356,-0.431],[358,0],[360,-0.431],[362,-0.431],[364,-0.1512],[366,0.0003],[368,0]]
       }
       hand_r_fk_ctrl.Location.Z {
        type=MovieSceneScriptingFloatChannel
        default=0
        thinned_from=93
        keys [[0,-0.0001],[94,-0.0002],[110,-0.0016],[112,-0.002],[116,-0.0035],[154,0],[220,0],[222,0.0002],[246,0.0004],[248,0],[250,0.0001],[252,0],[270,0],[272,0.0002],[274,0],[334,0],[336,0.0007],[340,0.0037],[342,0.0023],[344,0.0002],[346,0.0668],[348,0.0001],[350,0.0668],[352,0],[354,0],[356,0.0668],[358,0],[360,0.0668],[362,0.0668],[364,0.0234],[366,-0.0006],[368,0]]
       }
       hand_r_fk_ctrl.Rotation.X {
        type=MovieSceneScriptingFloatChannel
        default=12.2504
        thinned_from=93
        keys [[0,4.3196],[92,4.3196],[94,4.3521],[96,4.4235],[98,4.5318],[100,4.6749],[102,4.8506],[104,5.057],[106,5.2918],[108,5.5531],[110,5.8386],[112,6.1792],[114,6.591],[116,7.0503],[154,5.1028],[174,5.258],[176,5.5651],[178,5.584],[180,5.6755],[182,5.7909],[212,5.7852],[214,5.7605],[216,5.648],[218,5.0256],[220,3.8915],[222,2.1763],[246,14.9394],[248,15.2061],[250,13.9545],[252,13.0425],[254,13.2041],[258,13.314],[260,13.4737],[262,14.0273],[266,14.2833],[268,16.1666],[270,17.2761],[272,17.3963],[274,17.3086],[276,18.8077],[278,22.6202],[280,27.2375],[282,30.8358],[284,33.4877],[286,34.4749],[290,34.6997],[292,34.7127],[294,34.6388],[298,34.7644],[302,34.6028],[304,34.6234],[308,34.4283],[312,34.6699],[316,34.5731],[318,34.6789],[322,34.5509],[324,34.5684],[326,34.6802],[330,34.6508],[332,32.6549],[334,27.061],[336,19.6124],[338,17.0174],[340,17.8084],[342,18.0773],[344,19.3944],[346,24.8589],[348,23.0841],[350,23.0661],[352,24.1675],[354,24.0214],[356,23.4796],[358,24.8094],[360,26.7713],[362,-1.7878],[364,7.7667],[366,14.1221],[368,12.2504]]
       }
       hand_r_fk_ctrl.Rotation.Y {
        type=MovieSceneScriptingFloatChannel
        default=6.6226
        thinned_from=93
        keys [[0,7.7591],[92,7.7591],[94,7.5119],[96,7.2302],[98,6.9224],[106,5.5975],[108,5.2846],[110,4.9959],[112,4.1454],[114,2.4105],[116,0.194],[154,13.725],[174,14.5418],[176,14.725],[178,15.2387],[180,15.5211],[182,15.7272],[214,15.728],[216,15.7499],[218,15.8082],[220,15.9568],[222,16.3566],[246,-0.4571],[248,3.6333],[250,4.3986],[252,4.2518],[254,3.9427],[256,3.7273],[258,3.49],[260,3.1768],[262,2.9975],[264,2.8796],[266,1.7044],[268,2.5971],[270,2.9785],[272,2.9907],[274,2.3973],[276,0.9103],[278,-2.378],[280,-5.7013],[282,-7.3523],[284,-7.2264],[286,-6.3049],[288,-5.403],[290,-4.7159],[292,-4.252],[294,-3.9874],[296,-3.8627],[298,-3.8085],[300,-3.7931],[306,-3.7896],[308,-3.8053],[310,-3.8023],[312,-3.846],[314,-3.8615],[316,-4.0292],[318,-4.0476],[320,-4.2076],[322,-4.3449],[324,-4.5002],[326,-4.6227],[328,-4.498],[330,-4.5209],[332,-4.7145],[334,-5.6239],[336,-7.5042],[338,-4.8396],[340,0.7138],[342,6.7803],[344,12.8155],[346,-6.6251],[348,22.2684],[350,22.5601],[352,22.29],[354,22.4637],[356,21.1219],[358,24.4048],[360,22.049],[362,1.9879],[364,4.5049],[366,3.7398],[368,6.6226]]
       }
       hand_r_fk_ctrl.Rotation.Z {
        type=MovieSceneScriptingFloatChannel
        default=-34.4326
        thinned_from=93
        keys [[0,-32.1316],[92,-32.1316],[94,-31.4617],[96,-30.6522],[98,-29.7284],[100,-28.7157],[108,-24.2827],[110,-23.2057],[112,-20.6051],[114,-15.6172],[116,-9.3148],[154,-11.5408],[174,-12.9803],[176,-13.7904],[178,-14.4797],[180,-15.0275],[182,-15.4233],[212,-15.4308],[216,-15.4825],[218,-15.4546],[220,-15.4764],[222,-14.252],[246,-20.4781],[248,-25.1221],[250,-26.5768],[252,-27.0301],[254,-26.7033],[256,-26.8792],[258,-27.3897],[260,-27.616],[262,-27.7336],[264,-28.039],[266,-29.0157],[268,-28.4132],[270,-28.9622],[272,-30.7488],[274,-32.4716],[276,-33.8216],[278,-32.8333],[280,-29.9435],[282,-27.5729],[284,-25.4882],[286,-24.0562],[288,-23.404],[290,-23.118],[292,-22.7744],[294,-22.4867],[296,-22.3911],[300,-22.5952],[302,-22.4101],[306,-22.5819],[308,-22.5206],[310,-22.4112],[312,-22.5432],[314,-22.56],[316,-22.4614],[318,-22.4843],[324,-22.4015],[326,-22.3257],[330,-22.3052],[332,-24.5178],[334,-30.5972],[336,-40.4771],[338,-51.7108],[340,-51.2252],[342,-40.482],[344,-25.3975],[346,-6.3678],[348,-3.0974],[350,-3.3304],[352,-0.4574],[354,-2.396],[356,4.5759],[358,-10.5632],[360,-7.6197],[362,8.1643],[364,-21.5048],[366,-33.7404],[368,-34.4326]]
       }
       hand_r_fk_ctrl.Scale.X {type=MovieSceneScriptingFloatChannel, default=1, const [1,93]}
       hand_r_fk_ctrl.Scale.Y {type=MovieSceneScriptingFloatChannel, default=1, const [1,93]}
       hand_r_fk_ctrl.Scale.Z {type=MovieSceneScriptingFloatChannel, default=1, const [1,93]}
       thumb_01_r_ctrl.Location.X {
        type=MovieSceneScriptingFloatChannel
        default=0.0002
        thinned_from=67
        keys [[0,-0.0001],[222,-0.0001],[244,-0.0003],[266,-0.0001],[270,-0.0002],[272,-0.0011],[274,-0.0001],[276,-0.0009],[278,-0.0002],[280,-0.0003],[282,-0.0001],[332,0],[336,0.0002],[338,0.0002],[340,-0.0001],[342,-0.0002],[344,0.0003],[360,0.0002],[362,0],[364,0.0001],[366,0],[368,0.0002]]
       }
       thumb_01_r_ctrl.Location.Y {
        type=MovieSceneScriptingFloatChannel
        default=0.0005
        thinned_from=67
        keys [[0,0.0001],[222,0.0001],[270,0.0001],[272,-0.0006],[274,0.0001],[276,-0.0004],[278,0.0001],[280,-0.0001],[338,0.0002],[342,0.0001],[344,0.0004],[362,0.0006],[366,0.0004],[368,0.0005]]
       }
       thumb_01_r_ctrl.Location.Z {
        type=MovieSceneScriptingFloatChannel
        default=0.0002
        thinned_from=67
        keys [[0,0.0001],[222,0.0001],[244,-0.0008],[246,-0.0011],[248,-0.0004],[270,0.0001],[272,-0.0004],[274,0.0001],[276,-0.0001],[278,0.0001],[332,0.0002],[338,0.0004],[342,-0.0001],[344,0.0005],[362,0.0001],[366,0.0003],[368,0.0002]]
       }
       thumb_01_r_ctrl.Rotation.X {
        type=MovieSceneScriptingFloatChannel
        default=-9.7986
        thinned_from=67
        keys [[0,-1.5583],[222,-1.5583],[244,4.8386],[246,4.5289],[248,4.5781],[250,4.6922],[252,4.5719],[254,4.4778],[256,4.4293],[258,4.4252],[260,4.2152],[262,4.0622],[264,4.0388],[266,3.3244],[268,3.4721],[270,3.4212],[272,2.5382],[274,2.8353],[276,2.5717],[278,2.5554],[280,2.7094],[282,2.8918],[284,2.661],[286,2.5934],[288,2.6574],[290,2.4968],[292,2.4808],[294,2.5973],[296,2.6329],[298,2.675],[300,2.6425],[302,2.7633],[304,2.7442],[306,2.8319],[308,2.8753],[310,2.8052],[312,2.8326],[314,2.8489],[316,2.7747],[318,2.8581],[320,2.9016],[322,2.9252],[324,2.8997],[326,2.7743],[328,2.8395],[330,3.7812],[332,6.0752],[334,9.4328],[336,13.805],[338,17.9855],[340,20.2885],[342,19.4844],[344,15.4676],[346,10.3791],[348,4.0044],[350,-1.4239],[352,-5.9265],[354,-7.3942],[356,-7.4566],[358,-7.5664],[360,-7.6916],[362,-7.5248],[364,-9.4764],[366,-10.0125],[368,-9.7986]]
       }
       thumb_01_r_ctrl.Rotation.Y {
        type=MovieSceneScriptingFloatChannel
        default=9.1369
        thinned_from=67
        keys [[0,0.8707],[222,0.8707],[244,6.9278],[246,6.7669],[248,5.7775],[250,5.5362],[252,5.464],[254,5.4331],[256,5.3812],[258,5.5195],[260,5.5597],[262,5.6567],[264,5.7236],[266,5.9742],[268,6.1019],[270,6.2453],[272,6.4777],[274,6.3635],[276,6.3519],[278,6.2183],[280,6.2037],[282,6.1643],[284,6.3018],[286,6.3425],[288,6.4002],[290,6.4278],[292,6.3853],[294,6.4143],[296,6.4243],[298,6.3647],[302,6.1921],[304,6.178],[306,6.1894],[308,6.2359],[310,6.2463],[312,6.2846],[314,6.2779],[316,6.2973],[318,6.2751],[320,6.2881],[322,6.2651],[324,6.3325],[326,6.3095],[328,6.355],[330,6.116],[332,5.4972],[334,4.7693],[336,3.9613],[338,4.2526],[340,4.696],[342,6.093],[344,7.166],[346,8.1055],[348,9.0833],[350,9.5693],[352,10.015],[354,10.1672],[358,9.8721],[360,9.7053],[362,9.2412],[364,8.9089],[366,8.9839],[368,9.1369]]
       }
       thumb_01_r_ctrl.Rotation.Z {
        type=MovieSceneScriptingFloatChannel
        default=-10.5555
        thinned_from=67
        keys [[0,-13.9227],[222,-13.9227],[244,-16.2475],[246,-16.575],[248,-17.2395],[250,-17.5295],[252,-17.8979],[254,-18.1955],[256,-18.6549],[258,-18.8646],[262,-18.8958],[264,-18.8289],[266,-18.3809],[268,-18.1556],[270,-17.7258],[272,-16.8609],[274,-16.2945],[276,-16.1442],[278,-16.2159],[280,-16.4116],[282,-16.3406],[284,-16.383],[286,-16.3357],[288,-16.2292],[290,-16.0294],[296,-15.9962],[298,-16.0624],[304,-16.4449],[306,-16.5309],[308,-16.5194],[314,-16.3929],[318,-16.2453],[320,-16.2297],[322,-16.3267],[326,-16.2612],[328,-16.1979],[330,-16.5568],[332,-17.3597],[334,-17.0978],[336,-15.6583],[338,-15.9059],[340,-14.9903],[342,-14.2398],[344,-13.0712],[346,-11.9857],[348,-10.938],[350,-10.3503],[352,-10.0357],[354,-9.8041],[356,-9.6148],[358,-9.7968],[360,-10.0743],[362,-10.5608],[364,-10.8249],[366,-10.7106],[368,-10.5555]]
       }
       thumb_01_r_ctrl.Scale.X {type=MovieSceneScriptingFloatChannel, default=1, const [1,67]}
       thumb_01_r_ctrl.Scale.Y {type=MovieSceneScriptingFloatChannel, default=1, const [1,67]}
       thumb_01_r_ctrl.Scale.Z {type=MovieSceneScriptingFloatChannel, default=1, const [1,67]}
       thumb_02_r_ctrl.Location.X {
        type=MovieSceneScriptingFloatChannel
        default=0.0001
        thinned_from=67
        keys [[0,0],[222,0],[244,-0.0022],[246,-0.003],[248,-0.001],[250,-0.0007],[252,-0.0001],[266,0.0001],[268,0],[270,0.0001],[272,-0.0017],[274,0.0001],[276,-0.0015],[278,0],[280,-0.0004],[282,0.0001],[332,0.0003],[334,0.0008],[338,0.001],[340,0.0001],[342,-0.0003],[344,0.001],[354,0.0001],[360,0.0001],[362,-0.0005],[364,-0.0002],[366,-0.0006],[368,0.0001]]
       }
       thumb_02_r_ctrl.Location.Y {
        type=MovieSceneScriptingFloatChannel
        default=0.0005
        thinned_from=67
        keys [[0,0],[222,0],[246,-0.0001],[248,-0.0003],[266,0],[268,-0.0001],[270,0],[272,-0.001],[274,0],[276,-0.0011],[278,-0.0001],[280,-0.0004],[282,0],[332,0.0001],[336,0.0003],[338,0.0003],[342,-0.0002],[344,0.0004],[360,0.0005],[362,0.0002],[364,0.0004],[366,0.0001],[368,0.0005]]
       }
       thumb_02_r_ctrl.Location.Z {
        type=MovieSceneScriptingFloatChannel
        default=0.0002
        thinned_from=67
        keys [[0,0],[222,0],[244,-0.0005],[268,0.0001],[270,0],[272,-0.0007],[274,0],[330,0.0001],[338,0.0002],[342,-0.0001],[344,0.0003],[366,0.0004],[368,0.0002]]
       }
       thumb_02_r_ctrl.Rotation.X {
        type=MovieSceneScriptingFloatChannel
        default=0.0609
        thinned_from=67
        keys [[0,0],[222,0],[244,-0.7025],[246,-0.6547],[248,-0.4742],[250,-0.4325],[252,-0.4383],[254,-0.4406],[256,-0.4402],[258,-0.4463],[260,-0.46],[262,-0.4764],[264,-0.474],[266,-0.5079],[268,-0.5319],[270,-0.5316],[272,-0.574],[274,-0.4266],[276,-0.3891],[278,-0.3589],[280,-0.3666],[282,-0.366],[284,-0.3718],[286,-0.3766],[288,-0.3769],[290,-0.3726],[292,-0.3704],[296,-0.3715],[298,-0.3705],[300,-0.3685],[302,-0.3728],[304,-0.3753],[306,-0.3797],[308,-0.3777],[310,-0.3856],[312,-0.3808],[314,-0.3834],[316,-0.3823],[318,-0.3737],[320,-0.3783],[322,-0.3789],[324,-0.385],[326,-0.3845],[328,-0.3853],[330,-0.3157],[332,-0.1568],[334,0.0705],[336,0.2966],[338,0.5841],[340,0.7096],[342,0.7133],[344,0.7672],[346,0.8765],[348,0.8593],[350,0.9437],[352,0.9302],[356,0.9211],[358,0.918],[360,0.8506],[362,0.8437],[364,0.1148],[366,0.0719],[368,0.0609]]
       }
       thumb_02_r_ctrl.Rotation.Y {
        type=MovieSceneScriptingFloatChannel
        default=0.39
        thinned_from=67
        keys [[0,0],[222,0],[244,-1.9458],[246,-2.2724],[248,-4.0814],[250,-4.6862],[252,-4.9594],[254,-5.3273],[256,-5.6446],[258,-5.7664],[260,-5.642],[262,-5.5948],[264,-5.7093],[266,-4.785],[268,-5.0071],[270,-4.8474],[272,-3.487],[274,-4.1274],[276,-3.4975],[278,-3.2031],[280,-3.4346],[282,-3.4877],[284,-3.3652],[286,-3.3705],[288,-3.4196],[290,-3.2375],[292,-3.2],[294,-3.2703],[296,-3.2998],[298,-3.302],[300,-3.3238],[302,-3.471],[304,-3.5243],[306,-3.6055],[308,-3.569],[310,-3.5869],[312,-3.5508],[314,-3.5444],[316,-3.4728],[318,-3.4114],[320,-3.4679],[322,-3.5362],[324,-3.5552],[326,-3.5212],[328,-3.5106],[332,-3.3512],[334,-2.7807],[336,-1.2439],[338,0.3925],[340,1.8198],[342,2.8284],[344,2.7236],[346,1.6127],[348,0.9785],[350,-0.3069],[352,-0.8062],[354,-1.0758],[356,-1.0607],[358,-0.9855],[360,-0.8607],[362,-1.3696],[364,-0.1868],[366,0.587],[368,0.39]]
       }
       thumb_02_r_ctrl.Rotation.Z {
        type=MovieSceneScriptingFloatChannel
        default=-6.0293
        thinned_from=67
        keys [[0,0],[222,0],[244,17.546],[246,17.6398],[248,17.295],[252,16.4439],[254,15.8748],[256,15.3814],[258,15.1022],[260,15.0948],[264,14.7457],[266,15.7832],[268,15.0452],[270,15.3264],[272,16.9091],[274,17.7465],[276,18.7352],[278,19.1964],[280,18.8255],[282,18.7399],[284,18.9398],[286,18.9334],[288,18.8548],[290,19.1466],[292,19.2064],[294,19.0929],[296,19.0457],[300,19.0053],[302,18.7699],[304,18.6854],[306,18.5568],[308,18.6147],[310,18.5893],[312,18.6452],[314,18.6568],[318,18.8662],[322,18.6681],[324,18.6401],[328,18.7116],[330,18.6813],[332,18.3871],[334,18.6696],[336,19.5493],[338,20.9542],[340,21.2536],[342,19.8656],[344,16.5972],[346,12.1053],[348,7.1466],[350,2.7805],[352,-0.4259],[354,-1.7171],[356,-1.7222],[358,-1.676],[360,-1.9534],[362,-2.4615],[364,-6.3344],[366,-5.7291],[368,-6.0293]]
       }
       thumb_02_r_ctrl.Scale.X {type=MovieSceneScriptingFloatChannel, default=1, const [1,67]}
       thumb_02_r_ctrl.Scale.Y {type=MovieSceneScriptingFloatChannel, default=1, const [1,67]}
       thumb_02_r_ctrl.Scale.Z {type=MovieSceneScriptingFloatChannel, default=1, const [1,67]}
       thumb_03_r_ctrl.Location.X {
        type=MovieSceneScriptingFloatChannel
        default=0
        thinned_from=67
        keys [[0,0],[222,0],[244,-0.0009],[246,-0.0012],[248,-0.0004],[270,0],[272,-0.0006],[274,0],[276,-0.0006],[278,-0.0001],[280,-0.0003],[282,0],[332,0],[338,0.0003],[342,-0.0002],[344,0.0003],[360,0],[362,-0.0002],[366,-0.0002],[368,0]]
       }
       thumb_03_r_ctrl.Location.Y {type=MovieSceneScriptingFloatChannel, default=0.0001, thinned_from=67, keys [[0,0],[222,0],[246,0.0004],[248,0.0001],[368,0.0001]]}
       thumb_03_r_ctrl.Location.Z {
        type=MovieSceneScriptingFloatChannel
        default=0.0001
        thinned_from=67
        keys [[0,0],[222,0],[270,0],[272,-0.0003],[274,0],[342,-0.0001],[368,0.0001]]
       }
       thumb_03_r_ctrl.Rotation.X {
        type=MovieSceneScriptingFloatChannel
        default=2.2256
        thinned_from=67
        keys [[0,8.1645],[222,8.1645],[244,-1.0558],[246,-1.2807],[248,-1.3389],[250,-1.4245],[252,-1.4277],[254,-1.4235],[256,-1.5262],[258,-1.5469],[260,-1.5625],[262,-1.5994],[264,-1.6891],[266,-1.6193],[268,-1.7942],[270,-1.7702],[272,-1.5013],[274,-1.5811],[276,-1.3453],[278,-1.0668],[280,-1.1477],[284,-1.0833],[286,-1.134],[288,-1.165],[290,-1.1767],[292,-1.1395],[294,-1.166],[296,-1.1756],[298,-1.1505],[300,-1.1201],[302,-1.2002],[304,-1.1808],[306,-1.1924],[308,-1.1727],[310,-1.1905],[312,-1.1955],[314,-1.1976],[316,-1.1843],[318,-1.1371],[320,-1.1746],[322,-1.1816],[324,-1.2198],[326,-1.2071],[328,-1.1294],[330,-0.8889],[332,-0.5458],[334,-0.1332],[336,0.4221],[338,1.3014],[340,1.8243],[342,2.304],[344,2.6337],[346,2.8955],[348,3.1257],[350,3.3957],[352,3.5296],[354,3.5805],[356,3.5708],[358,3.5299],[360,3.5185],[362,3.3055],[364,2.1256],[366,2.3009],[368,2.2256]]
       }
       thumb_03_r_ctrl.Rotation.Y {
        type=MovieSceneScriptingFloatChannel
        default=0.9085
        thinned_from=67
        keys [[0,-9.9357],[222,-9.9357],[244,0.4065],[246,0.8204],[248,1.3657],[250,1.4244],[252,1.4202],[254,1.4326],[256,1.1985],[258,1.1527],[260,1.1188],[262,1.0341],[264,0.8202],[266,1.017],[268,0.5902],[270,0.6533],[272,1.3016],[274,1.1179],[276,1.657],[278,2.258],[280,2.0909],[284,2.2295],[286,2.1252],[288,2.0615],[290,2.0384],[292,2.1185],[294,2.0642],[296,2.0453],[298,2.0997],[300,2.1647],[302,1.997],[304,2.0393],[306,2.0157],[308,2.0585],[310,2.0216],[312,2.0118],[314,2.0079],[316,2.0368],[318,2.1368],[320,2.0583],[322,2.0437],[324,1.9625],[326,1.9896],[328,2.0071],[330,1.7971],[332,1.1271],[334,0.2239],[336,-1.0534],[338,-1.0892],[340,-0.8426],[342,-0.0195],[344,0.6927],[346,1.4232],[348,2.1683],[350,2.9809],[352,3.4727],[354,3.6644],[356,3.6472],[358,3.5751],[360,3.5549],[362,3.1725],[364,0.6535],[366,1.0772],[368,0.9085]]
       }
       thumb_03_r_ctrl.Rotation.Z {
        type=MovieSceneScriptingFloatChannel
        default=-50.5949
        thinned_from=67
        keys [[0,-50.7089],[222,-50.7089],[244,-29.2542],[246,-37.9951],[248,-43.797],[250,-46.4392],[254,-46.4813],[256,-47.0899],[260,-47.3476],[262,-47.5862],[264,-48.1403],[266,-47.9481],[268,-49.0164],[270,-48.8894],[272,-47.3261],[274,-47.8046],[276,-46.4954],[278,-45.0155],[280,-45.4601],[284,-45.1554],[286,-45.439],[288,-45.6196],[290,-45.6972],[292,-45.5148],[296,-45.7339],[300,-45.4652],[302,-45.9009],[304,-45.8081],[306,-45.8796],[308,-45.7836],[314,-45.938],[316,-45.8723],[318,-45.6267],[320,-45.8285],[322,-45.8683],[324,-46.0739],[326,-46.0066],[328,-44.2181],[330,-41.0105],[332,-38.4448],[334,-36.3409],[336,-28.8258],[338,-22.0025],[340,-22.4821],[342,-25.3327],[344,-29.3235],[346,-33.64],[348,-37.6537],[350,-40.5822],[352,-42.7996],[354,-43.6091],[356,-43.6542],[358,-43.8437],[360,-43.8968],[362,-44.8937],[364,-51.2136],[366,-50.176],[368,-50.5949]]
       }
       thumb_03_r_ctrl.Scale.X {type=MovieSceneScriptingFloatChannel, default=1, const [1,67]}
       thumb_03_r_ctrl.Scale.Y {type=MovieSceneScriptingFloatChannel, default=1, const [1,67]}
       thumb_03_r_ctrl.Scale.Z {type=MovieSceneScriptingFloatChannel, default=1, const [1,67]}
       index_metacarpal_r_ctrl.Location.X {
        type=MovieSceneScriptingFloatChannel
        default=0.0004
        thinned_from=67
        keys [[0,0.0001],[222,0.0001],[244,-0.0006],[246,-0.001],[248,-0.0004],[254,0],[270,0.0001],[272,-0.0007],[274,0.0001],[276,-0.0003],[278,0.0001],[280,-0.0001],[336,0.0001],[338,0.0003],[342,0],[344,0.0006],[360,0.0004],[362,0.0001],[364,0.0004],[368,0.0004]]
       }
       index_metacarpal_r_ctrl.Location.Y {
        type=MovieSceneScriptingFloatChannel
        default=0.0002
        thinned_from=67
        keys [[0,0.0002],[222,0.0002],[244,-0.0006],[246,-0.0009],[248,-0.0001],[272,0.0003],[274,0.0002],[276,0.0004],[278,0.0002],[340,0.0002],[342,0.0001],[344,0.0004],[364,0.0003],[366,0.0004],[368,0.0002]]
       }
       index_metacarpal_r_ctrl.Location.Z {
        type=MovieSceneScriptingFloatChannel
        default=-0.0006
        thinned_from=67
        keys [[0,0],[222,0],[246,0.0008],[254,0.0002],[266,0.0001],[268,0.0002],[270,0.0001],[272,0.0014],[274,0.0001],[276,0.001],[278,0.0001],[280,0.0004],[282,0.0001],[336,0.0001],[342,0.0002],[344,-0.0005],[366,-0.0004],[368,-0.0006]]
       }
       index_metacarpal_r_ctrl.Rotation.X {type=MovieSceneScriptingFloatChannel, default=0, thinned_from=67, keys [[0,0],[222,0],[244,0.01],[336,0.0099],[340,0],[368,0]]}
       index_metacarpal_r_ctrl.Rotation.Y {
        type=MovieSceneScriptingFloatChannel
        default=4.1035
        thinned_from=67
        keys [[0,3.7815],[222,3.7815],[244,-0.4748],[328,-0.4748],[330,-0.4732],[332,-0.4689],[336,-0.4559],[338,1.0887],[340,2.7723],[342,3.0557],[344,3.3273],[346,3.5757],[348,3.7892],[350,3.9561],[352,4.0647],[354,4.1035],[368,4.1035]]
       }
       index_metacarpal_r_ctrl.Rotation.Z {
        type=MovieSceneScriptingFloatChannel
        default=0
        thinned_from=67
        keys [[0,-3.094],[222,-3.094],[244,-0.74],[330,-0.7396],[336,-0.7352],[338,-0.3671],[340,0],[368,0]]
       }
       index_metacarpal_r_ctrl.Scale.X {type=MovieSceneScriptingFloatChannel, default=1, const [1,67]}
       index_metacarpal_r_ctrl.Scale.Y {type=MovieSceneScriptingFloatChannel, default=1, const [1,67]}
       index_metacarpal_r_ctrl.Scale.Z {type=MovieSceneScriptingFloatChannel, default=1, const [1,67]}
       index_01_r_ctrl.Location.X {
        type=MovieSceneScriptingFloatChannel
        default=0.0006
        thinned_from=67
        keys [[0,0],[222,0],[244,-0.0041],[246,-0.0056],[248,-0.0021],[250,-0.0016],[252,-0.0007],[254,-0.0003],[258,0],[266,0],[268,-0.0002],[270,0],[272,-0.0039],[274,0],[276,-0.0022],[278,-0.0001],[280,-0.0006],[282,0],[330,0],[332,-0.0002],[334,0],[336,-0.0003],[338,0.0005],[342,-0.0006],[344,0.0019],[352,0.0008],[360,0.0006],[362,-0.0003],[364,0.0002],[366,-0.0003],[368,0.0006]]
       }
       index_01_r_ctrl.Location.Y {
        type=MovieSceneScriptingFloatChannel
        default=0.0009
        thinned_from=67
        keys [[0,0],[222,0],[244,-0.0013],[246,-0.0013],[248,-0.0003],[266,0],[268,0.0002],[274,0],[276,0.0008],[278,0.0001],[280,0.0003],[282,0],[334,0],[342,-0.0001],[352,0.0009],[360,0.0009],[362,0.0007],[364,0.0012],[366,0.0012],[368,0.0009]]
       }
       index_01_r_ctrl.Location.Z {
        type=MovieSceneScriptingFloatChannel
        default=-0.0005
        thinned_from=67
        keys [[0,0],[222,0],[244,0.0002],[246,-0.0001],[248,0.0004],[270,0],[272,0.0003],[274,0],[276,0.0002],[278,0],[280,0.0002],[282,0],[336,0],[338,-0.0002],[342,0.0003],[344,-0.0005],[360,-0.0004],[362,-0.0001],[364,-0.0006],[368,-0.0005]]
       }
       index_01_r_ctrl.Rotation.X {
        type=MovieSceneScriptingFloatChannel
        default=3.8302
        thinned_from=67
        keys [[0,6.2821],[222,6.2821],[244,2.3068],[246,2.3559],[248,2.0759],[250,1.776],[252,1.4041],[254,1.1243],[256,0.7608],[258,0.5732],[260,0.4096],[262,0.1326],[264,-0.0725],[266,-0.2461],[268,-0.1257],[270,-0.104],[272,-0.3723],[274,-0.894],[276,-0.9974],[278,-0.8317],[280,-0.9194],[282,-0.7998],[284,-0.7097],[286,-0.7364],[288,-0.674],[290,-0.6852],[292,-0.6708],[294,-0.6444],[296,-0.6197],[298,-0.5546],[300,-0.5425],[302,-0.5422],[304,-0.4626],[306,-0.3443],[308,-0.2019],[310,-0.2255],[312,-0.2359],[314,-0.2475],[316,-0.2453],[318,-0.2411],[320,-0.2093],[322,-0.2046],[324,-0.259],[326,-0.2674],[328,-0.2596],[330,-0.2756],[332,-0.1539],[334,0.9979],[336,3.9884],[338,-0.7745],[340,-6.7249],[342,-4.8485],[344,-3.1804],[346,-1.1335],[348,0.86],[350,2.6822],[352,3.953],[354,4.4795],[356,4.5109],[358,4.5282],[360,4.5295],[362,4.4592],[364,3.5669],[366,3.6402],[368,3.8302]]
       }
       index_01_r_ctrl.Rotation.Y {
        type=MovieSceneScriptingFloatChannel
        default=-2.2962
        thinned_from=67
        keys [[0,-2.5276],[222,-2.5276],[244,-0.9212],[246,-1.379],[248,-2.0636],[250,-2.2119],[252,-2.2993],[254,-2.428],[256,-2.407],[258,-2.5151],[260,-2.5143],[262,-2.4519],[264,-2.3768],[266,-1.5381],[268,-2.0663],[270,-2.1141],[272,-1.8256],[274,-1.2788],[276,-1.0288],[278,-0.9637],[280,-0.9543],[282,-0.9571],[284,-0.9883],[286,-1.0101],[288,-1.0729],[290,-1.033],[292,-0.9692],[294,-0.9967],[296,-1.0138],[298,-1.0131],[300,-1.0814],[302,-1.0345],[304,-1.1228],[306,-1.2429],[308,-1.2546],[310,-1.2427],[312,-1.2191],[316,-1.2273],[318,-1.1427],[320,-1.1319],[322,-1.1483],[324,-1.1216],[326,-1.2019],[328,-1.1804],[330,-1.1509],[332,-1.1294],[334,-1.3045],[336,-1.3476],[338,2.4336],[340,6.4388],[342,5.3046],[344,4.028],[346,2.325],[348,0.4719],[350,-0.9862],[352,-2.3943],[354,-2.7389],[356,-2.6683],[358,-2.7441],[360,-2.6451],[362,-2.4774],[364,-2.2413],[366,-2.3415],[368,-2.2962]]
       }
       index_01_r_ctrl.Rotation.Z {
        type=MovieSceneScriptingFloatChannel
        default=-47.5191
        thinned_from=67
        keys [[0,-65.1155],[222,-65.1155],[244,-32.678],[246,-33.7359],[248,-33.5995],[250,-33.3689],[252,-32.9738],[254,-32.3402],[256,-31.5793],[260,-30.7614],[262,-29.8363],[264,-29.3996],[266,-30.1965],[268,-29.8606],[270,-28.7522],[272,-25.3119],[274,-19.7695],[276,-15.0748],[278,-10.9191],[280,-6.0383],[282,-2.522],[284,-0.1784],[286,0.8772],[288,0.83],[290,0.854],[292,0.7751],[294,0.7028],[296,0.6624],[298,0.3391],[300,0.343],[302,0.2953],[304,0.0568],[306,-0.3641],[308,-0.9349],[310,-0.7493],[312,-0.7763],[314,-0.7086],[316,-0.6914],[320,-0.8557],[322,-0.9015],[324,-0.791],[326,-0.6617],[328,-0.7097],[330,-1.9792],[332,-5.3124],[334,-10.9404],[336,-20.4139],[338,-15.2941],[340,-6.0811],[342,-16.5562],[344,-24.0429],[346,-32.6587],[348,-40.5108],[350,-47.1921],[352,-51.9928],[354,-53.9079],[356,-53.7365],[358,-53.7574],[360,-53.5933],[362,-52.8591],[364,-46.2829],[366,-46.7245],[368,-47.5191]]
       }
       index_01_r_ctrl.Scale.X {type=MovieSceneScriptingFloatChannel, default=1, const [1,67]}
       index_01_r_ctrl.Scale.Y {type=MovieSceneScriptingFloatChannel, default=1, const [1,67]}
       index_01_r_ctrl.Scale.Z {type=MovieSceneScriptingFloatChannel, default=1, const [1,67]}
       index_02_r_ctrl.Location.X {
        type=MovieSceneScriptingFloatChannel
        default=0.0001
        thinned_from=67
        keys [[0,-0.0001],[222,-0.0001],[244,-0.001],[246,-0.0014],[248,-0.0007],[250,-0.0005],[252,-0.0002],[266,-0.0001],[268,-0.0002],[270,-0.0001],[272,-0.001],[274,-0.0001],[276,-0.001],[278,-0.0001],[280,-0.0004],[282,-0.0001],[330,-0.0001],[332,-0.0005],[334,-0.0001],[336,-0.0006],[338,-0.0002],[340,-0.0001],[342,-0.0002],[344,0.0009],[354,0],[360,0],[362,-0.0005],[364,-0.0001],[368,0.0001]]
       }
       index_02_r_ctrl.Location.Y {
        type=MovieSceneScriptingFloatChannel
        default=-0.0002
        thinned_from=67
        keys [[0,-0.0001],[222,-0.0001],[246,-0.0014],[248,-0.0008],[254,-0.0002],[270,-0.0001],[272,-0.0016],[274,-0.0001],[276,-0.0008],[278,-0.0002],[280,-0.0003],[330,-0.0001],[332,-0.0004],[334,-0.0001],[336,-0.0004],[338,-0.0001],[340,-0.0001],[342,-0.0003],[344,0.0006],[354,-0.0002],[360,-0.0002],[362,-0.0006],[368,-0.0002]]
       }
       index_02_r_ctrl.Location.Z {
        type=MovieSceneScriptingFloatChannel
        default=-0.0001
        thinned_from=67
        keys [[0,-0.0001],[222,-0.0001],[244,-0.0007],[246,-0.001],[248,-0.0003],[272,0],[336,-0.0001],[338,0],[342,-0.0003],[344,0.0004],[360,-0.0001],[368,-0.0001]]
       }
       index_02_r_ctrl.Rotation.X {
        type=MovieSceneScriptingFloatChannel
        default=-0.1392
        thinned_from=67
        keys [[0,0],[222,0],[244,-1.0337],[246,-0.9252],[248,-0.8225],[252,-0.6701],[254,-0.6112],[256,-0.563],[258,-0.5218],[260,-0.5027],[262,-0.4643],[264,-0.44],[266,-0.4848],[268,-0.4624],[270,-0.4526],[272,-0.4216],[274,-0.4037],[276,-0.3885],[278,-0.3912],[280,-0.3824],[282,-0.3815],[284,-0.3906],[286,-0.3936],[288,-0.3948],[290,-0.3992],[292,-0.407],[294,-0.4075],[298,-0.4131],[300,-0.4106],[302,-0.4218],[304,-0.4198],[306,-0.4247],[310,-0.432],[312,-0.4302],[314,-0.4353],[316,-0.4346],[318,-0.4438],[320,-0.4481],[322,-0.4493],[324,-0.445],[326,-0.4421],[328,-0.4347],[330,-0.4411],[332,-0.4625],[334,-0.659],[336,-1.1906],[338,-0.8277],[340,-0.3154],[342,-0.3232],[344,-0.2449],[346,-0.1608],[348,-0.0711],[350,-0.0043],[352,0.0432],[354,0.0553],[356,0.0273],[358,0.0157],[360,-0.016],[362,-0.0439],[364,-0.1207],[366,-0.118],[368,-0.1392]]
       }
       index_02_r_ctrl.Rotation.Y {
        type=MovieSceneScriptingFloatChannel
        default=-0.0959
        thinned_from=67
        keys [[0,0],[222,0],[244,-0.0862],[246,-0.0965],[248,-0.1065],[250,-0.1203],[252,-0.1363],[254,-0.1497],[256,-0.1621],[258,-0.1717],[260,-0.1766],[262,-0.186],[264,-0.1931],[266,-0.1807],[268,-0.1865],[270,-0.1892],[274,-0.2032],[276,-0.208],[278,-0.207],[280,-0.2096],[282,-0.2101],[284,-0.2073],[286,-0.2063],[288,-0.206],[290,-0.2046],[292,-0.2023],[294,-0.2021],[298,-0.2005],[300,-0.2012],[302,-0.1979],[304,-0.1985],[310,-0.1949],[312,-0.1955],[314,-0.194],[316,-0.1942],[318,-0.1916],[320,-0.1904],[322,-0.1901],[326,-0.1921],[328,-0.1942],[330,-0.1918],[332,-0.1843],[334,-0.1404],[336,-0.1312],[338,0.089],[340,0.2317],[342,0.1501],[344,0.1221],[346,0.1004],[348,0.0934],[350,0.0843],[352,0.081],[354,0.0758],[356,0.0539],[358,0.0445],[360,0.018],[362,-0.0061],[364,-0.0775],[366,-0.0747],[368,-0.0959]]
       }
       index_02_r_ctrl.Rotation.Z {
        type=MovieSceneScriptingFloatChannel
        default=-72.4768
        thinned_from=67
        keys [[0,-60.8298],[222,-60.8298],[244,7.8699],[246,9.8785],[248,11.616],[250,12.596],[252,13.6848],[256,15.2653],[258,16.2177],[260,16.7763],[262,17.8849],[264,18.6036],[266,17.2873],[268,17.9297],[270,18.2268],[272,19.0791],[274,19.6602],[276,20.1189],[278,20.0281],[280,20.2792],[282,20.3172],[284,20.0488],[286,19.9585],[288,19.9228],[290,19.7951],[292,19.5628],[294,19.549],[298,19.3837],[300,19.4591],[302,19.1295],[304,19.1864],[310,18.8284],[312,18.8819],[314,18.7316],[316,18.7533],[318,18.4842],[320,18.3582],[322,18.322],[324,18.4476],[326,18.5348],[328,18.749],[330,18.5552],[332,17.9453],[334,12.207],[336,-2.8498],[338,-1.3562],[340,1.8998],[342,-5.749],[344,-15.6153],[346,-28.1772],[348,-41.0834],[350,-53.0482],[352,-61.6465],[354,-65.1576],[356,-66.1584],[358,-66.5798],[362,-68.7889],[364,-71.747],[366,-71.6359],[368,-72.4768]]
       }
       index_02_r_ctrl.Scale.X {type=MovieSceneScriptingFloatChannel, default=1, const [1,67]}
       index_02_r_ctrl.Scale.Y {type=MovieSceneScriptingFloatChannel, default=1, const [1,67]}
       index_02_r_ctrl.Scale.Z {type=MovieSceneScriptingFloatChannel, default=1, const [1,67]}
       index_03_r_ctrl.Location.X {
        type=MovieSceneScriptingFloatChannel
        default=0.0004
        thinned_from=67
        keys [[0,-0.0001],[222,-0.0001],[244,-0.0012],[246,-0.0017],[248,-0.0007],[250,-0.0006],[252,-0.0002],[266,-0.0001],[268,-0.0002],[270,-0.0001],[272,-0.0007],[274,-0.0001],[276,-0.0007],[278,-0.0001],[280,-0.0003],[282,-0.0001],[330,-0.0001],[332,-0.002],[334,-0.0001],[336,-0.0017],[338,-0.0006],[342,0.0005],[344,0.0019],[352,0.0005],[360,0.0004],[362,0.0001],[364,0.0003],[368,0.0004]]
       }
       index_03_r_ctrl.Location.Y {
        type=MovieSceneScriptingFloatChannel
        default=0.0001
        thinned_from=67
        keys [[0,0.0001],[222,0.0001],[246,-0.0009],[254,0],[270,0],[272,-0.0011],[274,0],[276,-0.0005],[278,0],[280,-0.0001],[330,0.0001],[332,-0.0008],[334,0.0001],[336,-0.0006],[342,0.0004],[344,0.0011],[352,0.0002],[360,0.0001],[362,0],[366,0.0002],[368,0.0001]]
       }
       index_03_r_ctrl.Location.Z {
        type=MovieSceneScriptingFloatChannel
        default=-0.0001
        thinned_from=67
        keys [[0,-0.0001],[222,-0.0001],[244,-0.0009],[246,-0.0012],[248,-0.0005],[270,-0.0001],[272,0.0001],[274,-0.0001],[334,-0.0001],[336,-0.0002],[338,0],[342,0],[344,0.0006],[354,-0.0001],[364,0],[368,-0.0001]]
       }
       index_03_r_ctrl.Rotation.X {
        type=MovieSceneScriptingFloatChannel
        default=-0.6094
        thinned_from=67
        keys [[0,0],[222,0],[244,-0.0493],[246,-0.0122],[248,0.026],[250,0.0501],[252,0.073],[254,0.0915],[256,0.1115],[258,0.1206],[260,0.1286],[262,0.1405],[264,0.1468],[266,0.1572],[268,0.1533],[270,0.1546],[272,0.1636],[274,0.1703],[276,0.1809],[278,0.1836],[280,0.1853],[284,0.184],[286,0.1852],[288,0.1827],[290,0.1831],[292,0.1812],[294,0.1815],[296,0.1796],[298,0.18],[300,0.182],[302,0.1778],[304,0.1774],[306,0.1762],[308,0.1768],[310,0.1745],[312,0.177],[314,0.1737],[316,0.174],[318,0.1707],[320,0.1683],[322,0.1692],[324,0.1707],[326,0.1728],[328,0.1734],[330,0.173],[332,0.1661],[334,0.0947],[336,-0.093],[338,-0.2699],[340,-0.4449],[342,-0.5429],[344,-0.56],[346,-0.568],[350,-0.5342],[352,-0.5247],[354,-0.5318],[356,-0.5476],[358,-0.5453],[360,-0.5613],[362,-0.5667],[364,-0.5954],[366,-0.5972],[368,-0.6094]]
       }
       index_03_r_ctrl.Rotation.Y {
        type=MovieSceneScriptingFloatChannel
        default=0.3765
        thinned_from=67
        keys [[0,0],[222,0],[244,-0.7626],[248,-0.8242],[252,-0.8665],[254,-0.8836],[256,-0.9034],[260,-0.9209],[262,-0.9331],[264,-0.9396],[266,-0.9513],[268,-0.9469],[270,-0.9485],[272,-0.956],[276,-0.9774],[280,-0.983],[284,-0.9816],[286,-0.9829],[288,-0.9801],[290,-0.9805],[292,-0.9783],[294,-0.9786],[296,-0.9765],[300,-0.9792],[302,-0.9744],[306,-0.9726],[308,-0.9732],[310,-0.9707],[312,-0.9735],[314,-0.9698],[316,-0.9701],[320,-0.9637],[328,-0.9694],[330,-0.9683],[332,-0.9593],[334,-0.8812],[336,-0.7194],[338,-0.0132],[340,0.6663],[342,0.6871],[344,0.6439],[346,0.5782],[348,0.4986],[350,0.4247],[352,0.3718],[354,0.3546],[356,0.3593],[358,0.3587],[360,0.3634],[362,0.365],[364,0.3728],[366,0.3733],[368,0.3765]]
       }
       index_03_r_ctrl.Rotation.Z {
        type=MovieSceneScriptingFloatChannel
        default=-26.1666
        thinned_from=67
        keys [[0,-78.806],[222,-78.806],[244,29.0125],[246,30.6758],[248,32.4176],[252,34.6041],[254,35.4754],[256,36.4539],[260,37.2966],[262,37.8879],[264,38.1997],[266,38.7401],[268,38.5353],[270,38.6094],[274,39.4215],[276,39.9508],[278,40.1185],[280,40.204],[284,40.1429],[286,40.2032],[288,40.0734],[290,40.0923],[292,39.992],[294,40.0073],[296,39.9084],[298,39.9293],[300,40.0347],[302,39.8137],[306,39.7295],[308,39.7605],[310,39.6415],[312,39.7699],[314,39.5996],[316,39.6167],[320,39.3193],[328,39.5813],[330,39.5555],[332,39.2169],[334,35.5645],[336,27.0061],[338,12.9319],[340,-2.3239],[342,-9.3534],[344,-13.408],[346,-17.0588],[348,-19.4532],[350,-21.2503],[352,-22.5223],[354,-23.369],[356,-23.9389],[358,-23.8576],[360,-24.4367],[362,-24.6325],[364,-25.6597],[366,-25.7229],[368,-26.1666]]
       }
       index_03_r_ctrl.Scale.X {type=MovieSceneScriptingFloatChannel, default=1, const [1,67]}
       index_03_r_ctrl.Scale.Y {type=MovieSceneScriptingFloatChannel, default=1, const [1,67]}
       index_03_r_ctrl.Scale.Z {type=MovieSceneScriptingFloatChannel, default=1, const [1,67]}
       middle_metacarpal_r_ctrl.Location.X {
        type=MovieSceneScriptingFloatChannel
        default=0.0002
        thinned_from=67
        keys [[0,0],[222,0],[244,-0.0012],[246,-0.0016],[248,-0.0007],[254,-0.0002],[270,-0.0002],[272,-0.0011],[274,-0.0001],[276,-0.0006],[278,-0.0001],[280,-0.0003],[336,-0.0001],[338,0.0001],[342,-0.0002],[344,0.0005],[360,0.0001],[362,-0.0001],[364,0.0001],[366,0],[368,0.0002]]
       }
       middle_metacarpal_r_ctrl.Location.Y {
        type=MovieSceneScriptingFloatChannel
        default=0.0003
        thinned_from=67
        keys [[0,0.0001],[246,0.0001],[270,0.0002],[272,0.0004],[274,0.0002],[276,0.0005],[278,0.0002],[366,0.0004],[368,0.0003]]
       }
       middle_metacarpal_r_ctrl.Location.Z {
        type=MovieSceneScriptingFloatChannel
        default=-0.0002
        thinned_from=67
        keys [[0,0],[222,0],[246,0.0001],[270,0.0001],[272,0.0004],[274,0.0001],[276,0.0003],[278,0.0001],[336,0.0001],[342,0.0001],[344,-0.0002],[368,-0.0002]]
       }
       middle_metacarpal_r_ctrl.Rotation.X {
        type=MovieSceneScriptingFloatChannel
        default=0
        thinned_from=67
        keys [[0,-7.29],[222,-7.29],[244,0.07],[332,0.0699],[336,0.0695],[340,0],[368,0]]
       }
       middle_metacarpal_r_ctrl.Rotation.Y {
        type=MovieSceneScriptingFloatChannel
        default=1.2252
        thinned_from=67
        keys [[0,1.3582],[222,1.3582],[244,1.7355],[330,1.7345],[336,1.7242],[338,0.861],[340,0],[342,0.0679],[344,0.2429],[346,0.4822],[348,0.743],[350,0.9823],[352,1.1574],[354,1.2252],[368,1.2252]]
       }
       middle_metacarpal_r_ctrl.Rotation.Z {
        type=MovieSceneScriptingFloatChannel
        default=0
        thinned_from=67
        keys [[0,-2.7048],[222,-2.7048],[244,-2.244],[330,-2.2427],[336,-2.2294],[338,-1.1133],[340,0],[368,0]]
       }
       middle_metacarpal_r_ctrl.Scale.X {type=MovieSceneScriptingFloatChannel, default=1, const [1,67]}
       middle_metacarpal_r_ctrl.Scale.Y {type=MovieSceneScriptingFloatChannel, default=1, const [1,67]}
       middle_metacarpal_r_ctrl.Scale.Z {type=MovieSceneScriptingFloatChannel, default=1, const [1,67]}
       middle_01_r_ctrl.Location.X {
        type=MovieSceneScriptingFloatChannel
        default=0.0007
        thinned_from=67
        keys [[0,0.0002],[222,0.0002],[244,-0.0036],[246,-0.0049],[248,-0.0017],[250,-0.0012],[252,-0.0004],[254,-0.0001],[266,0.0002],[268,0],[270,0.0001],[272,-0.0033],[274,0.0002],[276,-0.0018],[278,0.0001],[280,-0.0004],[282,0.0002],[330,0.0002],[332,0],[334,0.0002],[336,-0.0001],[338,0.0006],[342,-0.0004],[344,0.0018],[352,0.0008],[360,0.0007],[362,-0.0001],[364,0.0003],[366,-0.0001],[368,0.0007]]
       }
       middle_01_r_ctrl.Location.Y {
        type=MovieSceneScriptingFloatChannel
        default=0.0008
        thinned_from=67
        keys [[0,0],[222,0],[244,-0.0008],[246,-0.0007],[248,-0.0001],[266,0],[268,0.0002],[270,0],[274,0],[276,0.0007],[278,0.0001],[280,0.0003],[282,0],[338,-0.0001],[344,0],[352,0.0007],[362,0.0007],[364,0.001],[366,0.001],[368,0.0008]]
       }
       middle_01_r_ctrl.Location.Z {
        type=MovieSceneScriptingFloatChannel
        default=-0.0003
        thinned_from=67
        keys [[0,0],[244,0],[246,-0.0004],[248,0.0002],[272,-0.0001],[274,0],[276,-0.0001],[280,0.0001],[338,-0.0001],[342,0.0002],[344,-0.0003],[360,-0.0003],[362,0],[364,-0.0005],[366,-0.0005],[368,-0.0003]]
       }
       middle_01_r_ctrl.Rotation.X {
        type=MovieSceneScriptingFloatChannel
        default=3.7403
        thinned_from=67
        keys [[0,0],[222,0],[244,-8.1019],[246,-8.1394],[248,-8.087],[252,-8.0279],[254,-7.9714],[256,-7.9396],[264,-7.8704],[266,-7.8738],[268,-7.8461],[270,-7.9362],[272,-8.1621],[274,-8.4626],[278,-9.3199],[280,-9.6889],[282,-10.0243],[284,-10.2399],[286,-10.3172],[306,-10.348],[308,-10.374],[326,-10.3478],[328,-10.3577],[330,-9.9892],[332,-9.1928],[334,-8.4308],[336,-8.2063],[338,-9.0884],[340,-9.8477],[342,-9.4671],[344,-7.6179],[346,-5.0192],[348,-2.1019],[350,0.5908],[352,2.5831],[354,3.3303],[360,3.3575],[362,3.417],[364,3.823],[366,3.7949],[368,3.7403]]
       }
       middle_01_r_ctrl.Rotation.Y {
        type=MovieSceneScriptingFloatChannel
        default=-3.5248
        thinned_from=67
        keys [[0,1.3494],[222,1.3494],[244,-5.85],[246,-7.647],[248,-8.8375],[250,-9.2795],[252,-9.1552],[254,-9.0735],[256,-8.953],[258,-8.883],[260,-8.8584],[264,-8.6806],[266,-8.7708],[268,-8.7516],[270,-8.7841],[272,-8.7917],[274,-8.6427],[276,-8.6894],[278,-9.1016],[282,-9.7242],[284,-10.0582],[286,-10.1768],[288,-10.1717],[290,-10.0973],[294,-10.1548],[296,-10.1461],[298,-10.1933],[300,-10.1791],[302,-10.2218],[308,-10.4483],[312,-10.4494],[314,-10.4734],[316,-10.4311],[318,-10.4307],[320,-10.4704],[324,-10.4808],[326,-10.4537],[328,-10.1149],[330,-9.0299],[332,-7.3634],[334,-5.7183],[336,-3.021],[338,2.4871],[340,6.0018],[342,3.6904],[344,2.1987],[346,0.5177],[348,-1.0692],[350,-2.4109],[352,-3.4756],[354,-3.8153],[356,-3.7621],[358,-3.8077],[360,-3.8877],[362,-3.8607],[364,-3.1187],[366,-3.249],[368,-3.5248]]
       }
       middle_01_r_ctrl.Rotation.Z {
        type=MovieSceneScriptingFloatChannel
        default=-62.223
        thinned_from=67
        keys [[0,-74.7137],[222,-74.7137],[244,-44.1397],[246,-44.2403],[248,-42.7685],[250,-42.1289],[252,-41.2736],[254,-40.1741],[256,-39.4006],[258,-39.0277],[260,-38.7482],[264,-37.6461],[266,-37.7605],[268,-37.2285],[270,-36.3288],[272,-33.5625],[274,-28.6509],[276,-24.3899],[278,-20.653],[280,-16.0609],[282,-12.7135],[284,-10.3392],[286,-9.3854],[288,-9.3916],[290,-9.2607],[292,-9.3932],[294,-9.5041],[296,-9.5183],[298,-9.7326],[300,-9.703],[302,-9.7621],[304,-9.8446],[306,-10.198],[308,-10.7952],[310,-10.6571],[312,-10.6915],[316,-10.5352],[318,-10.533],[320,-10.6105],[322,-10.5942],[324,-10.5999],[326,-10.3405],[328,-10.5686],[330,-12.6948],[332,-17.705],[334,-24.2864],[336,-31.0024],[338,-26.3918],[340,-19.0767],[342,-29.3028],[344,-36.9844],[346,-45.8809],[348,-54.031],[350,-61.1095],[352,-66.1076],[354,-68.2988],[356,-68.0282],[360,-68.0238],[362,-67.1511],[364,-60.4717],[366,-61.0135],[368,-62.223]]
       }
       middle_01_r_ctrl.Scale.X {type=MovieSceneScriptingFloatChannel, default=1, const [1,67]}
       middle_01_r_ctrl.Scale.Y {type=MovieSceneScriptingFloatChannel, default=1, const [1,67]}
       middle_01_r_ctrl.Scale.Z {type=MovieSceneScriptingFloatChannel, default=1, const [1,67]}
       middle_02_r_ctrl.Location.X {
        type=MovieSceneScriptingFloatChannel
        default=0.001
        thinned_from=67
        keys [[0,0.0008],[222,0.0008],[244,0.0002],[246,-0.0001],[248,0.0004],[266,0.0008],[268,0.0006],[270,0.0007],[272,0.0003],[274,0.0008],[276,0.0001],[278,0.0007],[280,0.0005],[282,0.0008],[330,0.0008],[332,0.0005],[334,0.0008],[336,0.0005],[342,0.0008],[344,0.0014],[360,0.0009],[362,0.0004],[366,0.001],[368,0.001]]
       }
       middle_02_r_ctrl.Location.Y {
        type=MovieSceneScriptingFloatChannel
        default=0.0007
        thinned_from=67
        keys [[0,0.0009],[222,0.0009],[244,-0.0006],[246,-0.0009],[248,0],[254,0.0007],[266,0.0009],[268,0.0007],[270,0.0008],[272,-0.0009],[274,0.0009],[276,-0.0002],[278,0.0008],[280,0.0005],[282,0.0008],[330,0.0009],[332,0.0006],[334,0.0009],[336,0.0005],[338,0.0008],[340,0.0009],[342,0.0006],[344,0.0018],[352,0.0008],[360,0.0007],[362,0.0001],[364,0.0004],[368,0.0007]]
       }
       middle_02_r_ctrl.Location.Z {
        type=MovieSceneScriptingFloatChannel
        default=0.0476
        thinned_from=67
        keys [[0,0.0476],[222,0.0476],[244,0.0469],[246,0.0466],[248,0.0473],[270,0.0476],[272,0.0479],[274,0.0476],[332,0.0475],[334,0.0476],[336,0.0474],[338,0.0476],[342,0.0474],[344,0.0482],[356,0.0475],[368,0.0476]]
       }
       middle_02_r_ctrl.Rotation.X {
        type=MovieSceneScriptingFloatChannel
        default=-1.348
        thinned_from=67
        keys [[0,0.0065],[222,0.0065],[244,1.3125],[246,1.4023],[248,1.4431],[250,1.4519],[252,1.4458],[254,1.4319],[256,1.4092],[258,1.3837],[260,1.3707],[264,1.3202],[268,1.2899],[270,1.2805],[274,1.241],[276,1.2109],[278,1.2075],[280,1.1936],[282,1.1874],[284,1.19],[286,1.1867],[288,1.1894],[290,1.2042],[294,1.2085],[300,1.2084],[302,1.2149],[304,1.2152],[308,1.211],[310,1.2212],[312,1.2175],[316,1.2229],[320,1.2377],[322,1.2346],[324,1.228],[326,1.2297],[328,1.2213],[330,1.2263],[332,1.2534],[334,1.4177],[336,1.1908],[338,-0.8068],[340,-3.3131],[342,-3.6589],[344,-3.2166],[346,-2.4766],[348,-1.633],[350,-0.8163],[352,-0.2675],[354,-0.0592],[356,-0.2528],[358,-0.3416],[360,-0.5043],[362,-0.6424],[364,-1.2214],[366,-1.2633],[368,-1.348]]
       }
       middle_02_r_ctrl.Rotation.Y {
        type=MovieSceneScriptingFloatChannel
        default=-3.288
        thinned_from=67
        keys [[0,-0.0119],[222,-0.0119],[244,1.4386],[246,0.7053],[248,0.0831],[250,-0.3072],[252,-0.7587],[256,-1.3696],[258,-1.635],[260,-1.7492],[262,-1.9434],[264,-2.1219],[266,-2.2262],[268,-2.3126],[270,-2.3661],[272,-2.4843],[274,-2.5814],[276,-2.7331],[278,-2.7488],[280,-2.8165],[282,-2.8428],[284,-2.8308],[286,-2.8461],[288,-2.8337],[290,-2.7639],[294,-2.7435],[298,-2.7484],[300,-2.7441],[302,-2.7128],[304,-2.7112],[308,-2.7314],[310,-2.6817],[312,-2.6997],[314,-2.6832],[316,-2.6731],[320,-2.5986],[322,-2.6144],[324,-2.6474],[326,-2.6391],[328,-2.6813],[330,-2.6516],[332,-2.5072],[334,-1.1545],[336,1.9374],[338,2.4733],[340,2.003],[342,1.2899],[344,0.0033],[346,-1.3223],[348,-2.535],[350,-3.5543],[352,-4.1864],[354,-4.4114],[356,-4.22],[358,-4.1352],[360,-3.9844],[362,-3.8607],[364,-3.387],[366,-3.3529],[368,-3.288]]
       }
       middle_02_r_ctrl.Rotation.Z {
        type=MovieSceneScriptingFloatChannel
        default=-54.1794
        thinned_from=67
        keys [[0,-58.0462],[222,-58.0462],[244,20.6516],[246,24.252],[248,27.3235],[250,29.2331],[252,31.4392],[256,34.4338],[258,35.7397],[260,36.3001],[262,37.2589],[264,38.1405],[266,38.6568],[268,39.085],[270,39.3501],[272,39.9431],[274,40.4206],[276,41.1777],[278,41.2561],[280,41.595],[282,41.7261],[284,41.6662],[286,41.7429],[288,41.6804],[290,41.3315],[294,41.2294],[300,41.2325],[302,41.0759],[308,41.1689],[310,40.921],[312,41.0105],[316,40.878],[320,40.5064],[322,40.5852],[324,40.7497],[326,40.7082],[328,40.9186],[330,40.766],[332,40.0167],[334,33.3521],[336,18.1395],[338,9.8295],[340,1.9459],[342,-4.7555],[344,-11.6843],[346,-20.1806],[348,-29.1661],[350,-37.1111],[352,-43.2661],[354,-45.6841],[356,-47.0325],[358,-47.6425],[360,-48.7432],[362,-49.6616],[364,-53.3806],[366,-53.6443],[368,-54.1794]]
       }
       middle_02_r_ctrl.Scale.X {type=MovieSceneScriptingFloatChannel, default=1, const [1,67]}
       middle_02_r_ctrl.Scale.Y {type=MovieSceneScriptingFloatChannel, default=1, const [1,67]}
       middle_02_r_ctrl.Scale.Z {type=MovieSceneScriptingFloatChannel, default=1, const [1,67]}
       middle_03_r_ctrl.Location.X {
        type=MovieSceneScriptingFloatChannel
        default=0.001
        thinned_from=67
        keys [[0,0.0005],[222,0.0005],[244,-0.0008],[246,-0.0013],[248,-0.0001],[254,0.0004],[266,0.0005],[268,0.0004],[270,0.0005],[272,-0.0002],[274,0.0005],[276,-0.0001],[278,0.0005],[280,0.0003],[282,0.0005],[330,0.0005],[332,-0.0016],[334,0.0005],[336,-0.0009],[338,0],[340,0.0005],[342,0.0013],[344,0.0026],[352,0.0012],[360,0.001],[362,0.0007],[364,0.001],[368,0.001]]
       }
       middle_03_r_ctrl.Location.Y {
        type=MovieSceneScriptingFloatChannel
        default=0.0002
        thinned_from=67
        keys [[0,0],[222,0],[246,-0.0008],[248,-0.0005],[254,0],[270,0],[272,-0.0009],[274,0],[276,-0.0004],[278,0],[280,-0.0001],[330,0],[332,-0.0007],[334,0],[336,-0.0004],[340,0],[344,0.0009],[354,0.0001],[360,0.0001],[362,0],[366,0.0002],[368,0.0002]]
       }
       middle_03_r_ctrl.Location.Z {
        type=MovieSceneScriptingFloatChannel
        default=0.0061
        thinned_from=67
        keys [[0,0.0061],[222,0.0061],[244,0.0053],[246,0.0049],[248,0.0057],[270,0.0061],[272,0.0063],[330,0.0061],[332,0.0059],[334,0.0061],[336,0.0059],[338,0.0061],[342,0.0062],[344,0.0069],[354,0.0061],[362,0.006],[364,0.0062],[368,0.0061]]
       }
       middle_03_r_ctrl.Rotation.X {
        type=MovieSceneScriptingFloatChannel
        default=-1.0817
        thinned_from=67
        keys [[0,-0.007],[222,-0.007],[244,-1.3898],[246,-1.328],[248,-1.2771],[250,-1.2513],[252,-1.2285],[254,-1.2123],[256,-1.2006],[258,-1.1915],[266,-1.1746],[272,-1.1689],[328,-1.1702],[332,-1.1652],[334,-1.1955],[336,-1.3893],[338,-0.2961],[340,0.7143],[342,0.431],[344,0.2419],[346,0.0005],[348,-0.1823],[350,-0.3393],[352,-0.5117],[354,-0.5501],[356,-0.5854],[358,-0.6303],[360,-0.6688],[362,-0.7308],[364,-1.0033],[366,-1.0266],[368,-1.0817]]
       }
       middle_03_r_ctrl.Rotation.Y {
        type=MovieSceneScriptingFloatChannel
        default=1.2471
        thinned_from=67
        keys [[0,-0.0086],[222,-0.0086],[244,1.9163],[248,1.3222],[250,1.1464],[252,0.9665],[254,0.817],[256,0.6872],[258,0.5737],[260,0.5103],[262,0.4248],[266,0.2634],[268,0.2125],[270,0.1807],[272,0.1244],[274,0.0257],[276,-0.0631],[278,-0.0337],[280,-0.0522],[282,-0.0425],[284,-0.0241],[286,-0.0219],[288,-0.0192],[290,-0.0142],[292,-0.0068],[294,0.0041],[296,0.0043],[298,0.0223],[300,0.0153],[302,0.0331],[304,0.041],[306,0.054],[308,0.0661],[310,0.0705],[312,0.0734],[314,0.0962],[316,0.089],[318,0.1147],[320,0.136],[322,0.128],[324,0.1258],[326,0.1031],[328,0.1031],[330,0.1006],[332,0.1837],[334,0.7614],[336,1.9765],[338,2.3234],[340,2.2404],[342,2.169],[344,1.7939],[346,1.4596],[348,1.0633],[350,0.7316],[352,0.601],[354,0.5119],[356,0.5658],[358,0.6333],[360,0.6903],[362,0.7797],[364,1.1453],[366,1.1762],[368,1.2471]]
       }
       middle_03_r_ctrl.Rotation.Z {
        type=MovieSceneScriptingFloatChannel
        default=-50.6008
        thinned_from=67
        keys [[0,-64.5776],[222,-64.5776],[244,-3.2297],[246,-1.3112],[248,0.6202],[250,1.751],[252,2.9023],[254,3.8614],[256,4.6883],[258,5.4138],[260,5.8157],[262,6.3608],[266,7.3861],[268,7.7091],[270,7.9111],[272,8.2794],[274,8.894],[276,9.4558],[278,9.2709],[280,9.3886],[282,9.3263],[284,9.2096],[288,9.1787],[292,9.0999],[294,9.0304],[296,9.029],[298,8.9145],[300,8.959],[302,8.8458],[304,8.7957],[308,8.6362],[312,8.5899],[314,8.4453],[316,8.491],[318,8.3278],[320,8.1927],[322,8.2433],[324,8.2568],[326,8.4014],[328,8.4009],[330,8.4242],[332,7.9043],[334,4.2697],[336,-3.5313],[338,-0.2741],[340,3.9211],[342,-1.5102],[344,-8.5365],[346,-17.9538],[348,-27.2519],[350,-35.6603],[352,-42.489],[354,-44.8024],[356,-45.2139],[360,-46.1696],[362,-46.8645],[364,-49.781],[366,-50.0256],[368,-50.6008]]
       }
       middle_03_r_ctrl.Scale.X {type=MovieSceneScriptingFloatChannel, default=1, const [1,67]}
       middle_03_r_ctrl.Scale.Y {type=MovieSceneScriptingFloatChannel, default=1, const [1,67]}
       middle_03_r_ctrl.Scale.Z {type=MovieSceneScriptingFloatChannel, default=1, const [1,67]}
       ring_metacarpal_r_ctrl.Location.X {
        type=MovieSceneScriptingFloatChannel
        default=0.0003
        thinned_from=67
        keys [[0,0.0001],[222,0.0001],[244,-0.0012],[246,-0.0015],[248,-0.0005],[254,0],[270,0],[272,-0.001],[274,0.0001],[276,-0.0005],[278,0.0001],[280,-0.0001],[332,0.0002],[336,0.0005],[338,0.0004],[342,-0.0001],[344,0.0005],[364,0.0002],[366,0],[368,0.0003]]
       }
       ring_metacarpal_r_ctrl.Location.Y {
        type=MovieSceneScriptingFloatChannel
        default=0.0004
        thinned_from=67
        keys [[0,0.0002],[222,0.0002],[244,0.0006],[246,0.0009],[248,0.0005],[270,0.0002],[272,0.0005],[274,0.0002],[276,0.0006],[278,0.0003],[280,0.0004],[332,0.0001],[338,-0.0001],[342,0.0003],[344,-0.0001],[366,0.0004],[368,0.0004]]
       }
       ring_metacarpal_r_ctrl.Location.Z {
        type=MovieSceneScriptingFloatChannel
        default=0.0001
        thinned_from=67
        keys [[0,0],[222,0],[244,-0.0003],[246,-0.0006],[248,-0.0002],[270,-0.0001],[272,-0.0006],[274,0],[276,-0.0004],[278,0],[342,0],[362,0.0002],[366,0],[368,0.0001]]
       }
       ring_metacarpal_r_ctrl.Rotation.X {
        type=MovieSceneScriptingFloatChannel
        default=0
        thinned_from=67
        keys [[0,-2.4847],[222,-2.4847],[244,-0.04],[328,-0.04],[330,-0.037],[332,-0.0296],[336,-0.0104],[338,-0.003],[340,0],[368,0]]
       }
       ring_metacarpal_r_ctrl.Rotation.Y {
        type=MovieSceneScriptingFloatChannel
        default=-4.4979
        thinned_from=67
        keys [[0,-0.8637],[222,-0.8637],[244,3.1532],[328,3.1532],[330,3.0014],[332,2.5973],[334,2.018],[336,1.3406],[338,0.6423],[340,0],[342,-0.6816],[344,-1.4923],[346,-2.3468],[348,-3.16],[350,-3.8464],[352,-4.3208],[354,-4.4979],[368,-4.4979]]
       }
       ring_metacarpal_r_ctrl.Rotation.Z {
        type=MovieSceneScriptingFloatChannel
        default=0
        thinned_from=67
        keys [[0,-5.6698],[222,-5.6698],[244,-0.8],[328,-0.8],[330,-0.7407],[332,-0.5926],[336,-0.2074],[338,-0.0593],[340,0],[368,0]]
       }
       ring_metacarpal_r_ctrl.Scale.X {type=MovieSceneScriptingFloatChannel, default=1, const [1,67]}
       ring_metacarpal_r_ctrl.Scale.Y {type=MovieSceneScriptingFloatChannel, default=1, const [1,67]}
       ring_metacarpal_r_ctrl.Scale.Z {type=MovieSceneScriptingFloatChannel, default=1, const [1,67]}
       ring_01_r_ctrl.Location.X {
        type=MovieSceneScriptingFloatChannel
        default=0.0001
        thinned_from=67
        keys [[0,-0.0002],[222,-0.0002],[244,-0.0037],[246,-0.005],[248,-0.002],[250,-0.0015],[252,-0.0008],[254,-0.0004],[266,-0.0002],[268,-0.0004],[270,-0.0002],[272,-0.0035],[274,-0.0002],[276,-0.0022],[278,-0.0003],[280,-0.0007],[282,-0.0002],[328,-0.0002],[332,0.0002],[334,0.0008],[336,0.0011],[338,0.0011],[340,-0.0002],[342,-0.0007],[344,0.0013],[352,0.0003],[360,0.0001],[362,-0.0005],[364,-0.0002],[366,-0.0007],[368,0.0001]]
       }
       ring_01_r_ctrl.Location.Y {
        type=MovieSceneScriptingFloatChannel
        default=0.0008
        thinned_from=67
        keys [[0,0],[222,0],[246,-0.0012],[248,-0.0003],[268,0.0001],[270,0],[272,-0.0005],[276,0.0003],[278,0],[280,0.0002],[334,-0.0001],[338,-0.0001],[354,0.0008],[360,0.0008],[362,0.0006],[364,0.0008],[368,0.0008]]
       }
       ring_01_r_ctrl.Location.Z {
        type=MovieSceneScriptingFloatChannel
        default=-0.0003
        thinned_from=67
        keys [[0,0],[244,0.0001],[246,-0.0003],[248,0.0002],[270,0],[272,-0.0003],[274,0],[276,-0.0003],[278,0],[338,-0.0002],[342,0.0002],[344,-0.0002],[360,-0.0002],[362,0.0001],[364,-0.0005],[366,-0.0005],[368,-0.0003]]
       }
       ring_01_r_ctrl.Rotation.X {
        type=MovieSceneScriptingFloatChannel
        default=-3.791
        thinned_from=67
        keys [[0,-4.509],[222,-4.509],[244,-9.9255],[246,-10.1807],[248,-10.6529],[250,-10.8126],[252,-10.8964],[264,-11.0001],[266,-10.9993],[272,-11.2001],[276,-11.4493],[278,-11.5142],[280,-11.5316],[282,-11.5932],[284,-11.5653],[300,-11.6281],[304,-11.5971],[308,-11.6273],[318,-11.5933],[322,-11.5651],[328,-11.577],[330,-11.1696],[332,-10.0884],[334,-8.5181],[336,-6.6894],[338,-4.9498],[340,-4.061],[342,-3.8597],[344,-3.8446],[346,-3.9686],[348,-4.0121],[350,-4.0906],[354,-4.1416],[356,-4.0841],[358,-4.0922],[360,-4.0298],[362,-4.0156],[364,-3.7962],[366,-3.8165],[368,-3.791]]
       }
       ring_01_r_ctrl.Rotation.Y {
        type=MovieSceneScriptingFloatChannel
        default=4.1677
        thinned_from=67
        keys [[0,-1.201],[222,-1.201],[244,-4.2256],[246,-4.7059],[248,-5.9954],[250,-6.5227],[252,-6.9723],[254,-7.0913],[256,-7.3603],[258,-7.5493],[260,-7.7539],[262,-7.8639],[264,-8.0218],[266,-8.2631],[268,-8.0817],[270,-8.0163],[272,-8.1121],[274,-8.3135],[276,-8.6677],[278,-8.9607],[280,-9.0984],[282,-9.1776],[284,-9.1016],[286,-9.1966],[288,-9.131],[294,-9.1707],[296,-9.1567],[298,-9.17],[300,-9.2622],[302,-9.2456],[304,-9.2034],[306,-9.2739],[308,-9.2995],[310,-9.2094],[312,-9.2116],[314,-9.1942],[316,-9.2099],[318,-9.181],[320,-9.0927],[322,-9.1036],[324,-9.1356],[328,-9.1323],[330,-8.8182],[332,-7.8454],[334,-5.3648],[336,-0.4782],[338,2.2393],[340,3.1542],[342,3.1616],[344,3.3774],[346,3.2419],[348,3.5541],[352,3.7207],[354,3.7565],[356,3.9781],[358,3.7709],[360,3.9034],[362,3.8956],[364,4.5253],[366,4.4322],[368,4.1677]]
       }
       ring_01_r_ctrl.Rotation.Z {
        type=MovieSceneScriptingFloatChannel
        default=-62.6302
        thinned_from=67
        keys [[0,-79.0231],[222,-79.0231],[244,-50.5644],[248,-55.6323],[250,-56.6081],[252,-56.9138],[254,-56.565],[256,-56.7009],[258,-57.188],[264,-57.5104],[266,-57.4854],[268,-57.3331],[270,-57.7063],[272,-58.2419],[274,-58.497],[278,-60.9208],[282,-61.7909],[286,-62.084],[288,-62.1047],[290,-61.9358],[304,-62.2295],[306,-62.503],[308,-62.9152],[310,-62.7139],[320,-62.3304],[324,-62.502],[326,-62.3022],[328,-62.557],[330,-59.8788],[332,-53.0629],[334,-42.7837],[336,-29.0411],[338,-25.2319],[340,-25.2264],[342,-33.1437],[344,-40.3357],[346,-49.3077],[348,-57.0866],[350,-64.1533],[352,-68.9606],[354,-71.1322],[356,-70.5334],[358,-70.4796],[360,-70.2772],[362,-69.1117],[364,-61.1381],[366,-61.2004],[368,-62.6302]]
       }
       ring_01_r_ctrl.Scale.X {type=MovieSceneScriptingFloatChannel, default=1, const [1,67]}
       ring_01_r_ctrl.Scale.Y {type=MovieSceneScriptingFloatChannel, default=1, const [1,67]}
       ring_01_r_ctrl.Scale.Z {type=MovieSceneScriptingFloatChannel, default=1, const [1,67]}
       ring_02_r_ctrl.Location.X {
        type=MovieSceneScriptingFloatChannel
        default=0.0004
        thinned_from=67
        keys [[0,0.0003],[266,0.0003],[276,0.0001],[332,0.0003],[338,0.0005],[342,0.0001],[344,0.0005],[360,0.0003],[362,-0.0002],[366,0.0003],[368,0.0004]]
       }
       ring_02_r_ctrl.Location.Y {
        type=MovieSceneScriptingFloatChannel
        default=0
        thinned_from=67
        keys [[0,0.0001],[222,0.0001],[244,-0.0008],[246,-0.0011],[248,-0.0005],[252,-0.0001],[266,0],[268,-0.0001],[270,0],[272,-0.0008],[274,0],[276,-0.0007],[278,0],[280,-0.0002],[282,0],[332,0.0002],[334,0.0005],[338,0.0006],[342,-0.0003],[344,0.0007],[354,0],[360,-0.0001],[362,-0.0006],[368,0]]
       }
       ring_02_r_ctrl.Location.Z {
        type=MovieSceneScriptingFloatChannel
        default=-0.0001
        thinned_from=67
        keys [[0,0],[222,0],[244,-0.0003],[246,-0.0005],[248,-0.0001],[270,0],[272,0.0005],[274,0],[276,0.0002],[278,0],[338,0.0002],[342,-0.0002],[344,0.0002],[362,-0.0001],[368,-0.0001]]
       }
       ring_02_r_ctrl.Rotation.X {
        type=MovieSceneScriptingFloatChannel
        default=1.0374
        thinned_from=67
        keys [[0,0],[222,0],[244,-5.0123],[246,-4.9394],[248,-5.4832],[250,-5.5198],[252,-5.3806],[254,-5.3275],[258,-5.0801],[260,-5.0437],[266,-4.8735],[268,-4.8256],[270,-4.8117],[272,-4.8151],[274,-4.7648],[276,-4.7301],[278,-4.8011],[280,-4.7548],[282,-4.7707],[284,-4.7471],[288,-4.7526],[290,-4.815],[292,-4.8338],[296,-4.8137],[300,-4.8566],[302,-4.8442],[304,-4.8726],[306,-4.8674],[308,-4.8509],[310,-4.8633],[312,-4.8362],[316,-4.8574],[318,-4.888],[320,-4.8997],[324,-4.8765],[326,-4.9317],[328,-4.8894],[330,-4.387],[334,-1.9665],[336,-1.1259],[338,-0.1148],[340,0],[342,0.201],[344,0.7441],[346,1.3196],[348,1.815],[350,2.239],[352,2.4214],[354,2.535],[356,2.2371],[358,2.1517],[360,1.9759],[362,1.846],[364,1.2187],[366,1.1066],[368,1.0374]]
       }
       ring_02_r_ctrl.Rotation.Y {
        type=MovieSceneScriptingFloatChannel
        default=0.5176
        thinned_from=67
        keys [[0,0],[222,0],[244,6.4488],[246,6.4201],[248,6.6141],[250,6.6254],[252,6.5805],[254,6.563],[258,6.4752],[268,6.3753],[272,6.3717],[276,6.3334],[278,6.365],[280,6.346],[282,6.3522],[288,6.3445],[290,6.3707],[296,6.3702],[300,6.3879],[302,6.3828],[304,6.3944],[310,6.3906],[312,6.3795],[320,6.4054],[324,6.396],[326,6.4181],[328,6.4012],[330,5.8825],[332,4.614],[334,3.1192],[336,1.6866],[338,0.4842],[340,0],[342,0.0246],[344,0.145],[346,0.3179],[348,0.5078],[350,0.6867],[352,0.8061],[354,0.8586],[356,0.815],[358,0.8003],[360,0.7671],[362,0.7402],[364,0.5709],[366,0.54],[368,0.5176]]
       }
       ring_02_r_ctrl.Rotation.Z {
        type=MovieSceneScriptingFloatChannel
        default=-62.8338
        thinned_from=67
        keys [[0,-66.6687],[222,-66.6687],[244,-16.3335],[246,-14.4656],[248,-17.4243],[250,-16.5799],[252,-14.5185],[254,-13.2447],[256,-11.5653],[258,-10.0997],[260,-9.4854],[262,-8.9443],[264,-8.4635],[266,-8.0418],[268,-7.6664],[270,-7.5572],[272,-7.5786],[274,-7.1877],[276,-6.9166],[278,-7.4733],[280,-7.1081],[282,-7.2342],[284,-7.0477],[286,-7.049],[288,-7.0915],[290,-7.5825],[292,-7.7303],[296,-7.5725],[298,-7.7637],[300,-7.9094],[302,-7.812],[304,-8.0347],[306,-7.9942],[308,-7.8646],[310,-7.9615],[312,-7.7491],[314,-7.8412],[316,-7.9156],[318,-8.1555],[320,-8.2472],[322,-8.1909],[324,-8.0649],[326,-8.4973],[328,-8.1666],[330,-6.5231],[332,-3.0301],[334,-1.7326],[336,-4.0121],[338,-4.5668],[340,-10.1255],[342,-15.8156],[344,-21.9101],[346,-29.2343],[348,-37.1203],[350,-44.0497],[352,-49.8054],[354,-51.662],[356,-53.8469],[358,-54.4775],[360,-55.7787],[362,-56.7401],[364,-61.4615],[366,-62.3036],[368,-62.8338]]
       }
       ring_02_r_ctrl.Scale.X {type=MovieSceneScriptingFloatChannel, default=1, const [1,67]}
       ring_02_r_ctrl.Scale.Y {type=MovieSceneScriptingFloatChannel, default=1, const [1,67]}
       ring_02_r_ctrl.Scale.Z {type=MovieSceneScriptingFloatChannel, default=1, const [1,67]}
       ring_03_r_ctrl.Location.X {
        type=MovieSceneScriptingFloatChannel
        default=0.0004
        thinned_from=67
        keys [[0,0],[222,0],[244,-0.0004],[246,-0.0006],[248,-0.0002],[270,0],[272,-0.0005],[274,0],[332,0.0001],[334,0.0005],[336,0.0007],[338,0.0007],[340,0],[342,-0.0003],[344,0.0009],[360,0.0004],[362,0.0001],[368,0.0004]]
       }
       ring_03_r_ctrl.Location.Y {
        type=MovieSceneScriptingFloatChannel
        default=0.0001
        thinned_from=67
        keys [[0,-0.0001],[222,-0.0001],[270,-0.0001],[272,0.0002],[274,-0.0001],[276,0],[342,-0.0002],[360,0.0001],[362,0],[368,0.0001]]
       }
       ring_03_r_ctrl.Location.Z {
        type=MovieSceneScriptingFloatChannel
        default=0
        thinned_from=67
        keys [[0,0],[222,0],[246,-0.0005],[248,-0.0001],[270,0],[272,0.0003],[274,0],[332,0.0001],[336,0.0004],[338,0.0004],[340,0],[342,-0.0001],[344,0.0005],[362,-0.0001],[364,0.0001],[368,0]]
       }
       ring_03_r_ctrl.Rotation.X {
        type=MovieSceneScriptingFloatChannel
        default=0.4078
        thinned_from=67
        keys [[0,0],[222,0],[244,-0.864],[246,-0.8157],[248,-0.8078],[250,-0.7799],[254,-0.7449],[256,-0.7199],[258,-0.7158],[260,-0.7083],[262,-0.7027],[264,-0.6948],[266,-0.6799],[268,-0.6832],[270,-0.6884],[272,-0.6827],[274,-0.7119],[276,-0.724],[278,-0.7147],[280,-0.7177],[282,-0.717],[284,-0.7201],[286,-0.7112],[288,-0.7168],[290,-0.7144],[292,-0.7085],[294,-0.7116],[296,-0.7071],[298,-0.7143],[300,-0.7111],[302,-0.7023],[304,-0.7066],[306,-0.706],[310,-0.6997],[312,-0.6986],[314,-0.7005],[316,-0.7003],[318,-0.7117],[320,-0.7119],[322,-0.7056],[324,-0.7119],[328,-0.6961],[330,-0.6459],[332,-0.4724],[334,-0.2783],[336,-0.0933],[338,0.0516],[340,0],[342,-0.024],[344,0.093],[346,0.173],[348,0.3341],[350,0.5064],[352,0.5706],[354,0.6271],[356,0.654],[358,0.5924],[360,0.6056],[362,0.5545],[364,0.4317],[366,0.4442],[368,0.4078]]
       }
       ring_03_r_ctrl.Rotation.Y {
        type=MovieSceneScriptingFloatChannel
        default=-0.4398
        thinned_from=67
        keys [[0,0],[222,0],[244,2.9486],[246,2.8548],[248,2.838],[250,2.7815],[254,2.7092],[256,2.6562],[258,2.6474],[264,2.6022],[266,2.5695],[270,2.5881],[272,2.5773],[274,2.6391],[276,2.6647],[278,2.6451],[280,2.6514],[282,2.65],[284,2.6567],[286,2.6374],[288,2.6495],[290,2.6445],[292,2.6317],[294,2.6385],[296,2.6287],[298,2.6443],[300,2.6374],[302,2.6185],[304,2.6276],[312,2.6105],[316,2.614],[318,2.6385],[320,2.6391],[322,2.6256],[324,2.6391],[328,2.6051],[330,2.4141],[332,1.8362],[334,1.1596],[336,0.5096],[338,0.0108],[340,0],[342,0.0298],[344,-0.1291],[346,-0.2153],[348,-0.399],[350,-0.5917],[352,-0.6495],[354,-0.7116],[356,-0.7466],[358,-0.667],[360,-0.6838],[362,-0.6191],[364,-0.4683],[366,-0.4844],[368,-0.4398]]
       }
       ring_03_r_ctrl.Rotation.Z {
        type=MovieSceneScriptingFloatChannel
        default=-50.708
        thinned_from=67
        keys [[0,-73.2004],[222,-73.2004],[244,-2.8221],[246,-0.7631],[248,0.6001],[250,2.1797],[252,3.4702],[254,4.6335],[256,5.7652],[258,6.366],[260,6.8256],[262,7.038],[264,7.1732],[266,7.4323],[268,7.375],[270,7.2846],[272,7.3615],[274,6.8801],[276,6.679],[278,6.8322],[280,6.7814],[282,6.793],[284,6.7395],[286,6.8928],[288,6.7974],[290,6.837],[292,6.9385],[294,6.8847],[296,6.9629],[298,6.8384],[300,6.8932],[302,7.044],[304,6.9711],[306,6.9814],[310,7.0895],[312,7.1076],[314,7.0752],[316,7.0796],[318,6.8845],[320,6.8798],[322,6.9873],[324,6.8803],[328,7.15],[330,8.6885],[332,13.2314],[334,18.3348],[336,22.8229],[338,25.5272],[340,23.4929],[342,17.0665],[344,6.5894],[346,-7.8055],[348,-22.1554],[350,-34.8321],[352,-44.7379],[354,-48.1812],[356,-47.8627],[358,-48.59],[360,-48.4356],[362,-49.0314],[364,-50.4424],[366,-50.289],[368,-50.708]]
       }
       ring_03_r_ctrl.Scale.X {type=MovieSceneScriptingFloatChannel, default=1, const [1,67]}
       ring_03_r_ctrl.Scale.Y {type=MovieSceneScriptingFloatChannel, default=1, const [1,67]}
       ring_03_r_ctrl.Scale.Z {type=MovieSceneScriptingFloatChannel, default=1, const [1,67]}
       pinky_metacarpal_r_ctrl.Location.X {
        type=MovieSceneScriptingFloatChannel
        default=0.0001
        thinned_from=67
        keys [[0,-0.0001],[222,-0.0001],[246,-0.0014],[248,-0.0006],[254,-0.0002],[266,-0.0001],[270,-0.0001],[272,-0.0012],[274,-0.0001],[276,-0.0008],[278,-0.0001],[280,-0.0003],[332,0],[336,0.0003],[342,-0.0001],[344,0.0003],[364,-0.0001],[366,-0.0003],[368,0.0001]]
       }
       pinky_metacarpal_r_ctrl.Location.Y {
        type=MovieSceneScriptingFloatChannel
        default=0.0003
        thinned_from=67
        keys [[0,0],[222,0],[244,0.0012],[246,0.0016],[248,0.0006],[250,0.0004],[252,0.0001],[270,0.0001],[272,0.0002],[274,0],[276,0.0003],[278,0.0001],[280,0.0002],[330,0],[336,-0.0005],[338,-0.0005],[340,0],[342,0.0002],[344,-0.0004],[348,0],[350,0],[352,0.0002],[360,0.0002],[362,0.0004],[368,0.0003]]
       }
       pinky_metacarpal_r_ctrl.Location.Z {
        type=MovieSceneScriptingFloatChannel
        default=0.0005
        thinned_from=67
        keys [[0,0],[222,0],[244,-0.001],[246,-0.0015],[248,-0.0006],[254,-0.0001],[266,0],[268,-0.0002],[270,0],[272,-0.0015],[274,0],[276,-0.001],[278,-0.0001],[280,-0.0003],[282,0],[330,0],[336,0.0005],[338,0.0004],[340,0],[342,-0.0001],[344,0.0006],[362,0.0006],[364,0.0003],[366,0.0003],[368,0.0005]]
       }
       pinky_metacarpal_r_ctrl.Rotation.X {
        type=MovieSceneScriptingFloatChannel
        default=0
        thinned_from=67
        keys [[0,-5.9588],[222,-5.9588],[244,-0.58],[328,-0.58],[330,-0.537],[332,-0.4296],[336,-0.1504],[338,-0.043],[340,0],[368,0]]
       }
       pinky_metacarpal_r_ctrl.Rotation.Y {
        type=MovieSceneScriptingFloatChannel
        default=0
        thinned_from=67
        keys [[0,-5.2205],[222,-5.2205],[244,7.1566],[328,7.1566],[330,6.6265],[332,5.3012],[336,1.8554],[338,0.5301],[340,0],[368,0]]
       }
       pinky_metacarpal_r_ctrl.Rotation.Z {
        type=MovieSceneScriptingFloatChannel
        default=0
        thinned_from=67
        keys [[0,-11.0008],[222,-11.0008],[244,-4.63],[328,-4.63],[330,-4.287],[332,-3.4296],[336,-1.2004],[338,-0.343],[340,0],[368,0]]
       }
       pinky_metacarpal_r_ctrl.Scale.X {type=MovieSceneScriptingFloatChannel, default=1, const [1,67]}
       pinky_metacarpal_r_ctrl.Scale.Y {type=MovieSceneScriptingFloatChannel, default=1, const [1,67]}
       pinky_metacarpal_r_ctrl.Scale.Z {type=MovieSceneScriptingFloatChannel, default=1, const [1,67]}
       pinky_01_r_ctrl.Location.X {
        type=MovieSceneScriptingFloatChannel
        default=0.0003
        thinned_from=67
        keys [[0,0.0001],[222,0.0001],[244,-0.0034],[246,-0.0047],[248,-0.0016],[250,-0.0012],[252,-0.0004],[254,-0.0001],[266,0.0001],[268,-0.0001],[270,0.0001],[272,-0.0032],[274,0.0001],[276,-0.0019],[278,0],[280,-0.0004],[282,0.0001],[328,0.0001],[332,0.0005],[334,0.0011],[336,0.0014],[338,0.0014],[340,0.0001],[342,-0.0004],[344,0.0016],[352,0.0005],[360,0.0004],[362,-0.0003],[364,-0.0001],[366,-0.0005],[368,0.0003]]
       }
       pinky_01_r_ctrl.Location.Y {
        type=MovieSceneScriptingFloatChannel
        default=0.0011
        thinned_from=67
        keys [[0,0.0002],[222,0.0002],[246,-0.0013],[248,-0.0001],[270,0.0002],[272,-0.0006],[274,0.0002],[280,0.0004],[334,0.0002],[336,0.0001],[342,0.0002],[354,0.0011],[360,0.001],[362,0.0009],[364,0.001],[366,0.0009],[368,0.0011]]
       }
       pinky_01_r_ctrl.Location.Z {
        type=MovieSceneScriptingFloatChannel
        default=-0.0003
        thinned_from=67
        keys [[0,0],[222,0],[244,0.0004],[246,0.0001],[248,0.0002],[268,-0.0001],[270,0],[272,-0.0001],[274,0],[276,-0.0003],[278,0],[338,-0.0002],[342,0.0002],[344,-0.0003],[360,-0.0003],[362,0.0002],[364,-0.0005],[368,-0.0003]]
       }
       pinky_01_r_ctrl.Rotation.X {
        type=MovieSceneScriptingFloatChannel
        default=-2.7518
        thinned_from=67
        keys [[0,-11.7646],[222,-11.7646],[244,-10.3624],[246,-10.4757],[248,-10.715],[254,-10.8809],[256,-10.8375],[258,-10.7485],[264,-10.668],[266,-10.6089],[268,-10.7667],[270,-10.8167],[272,-10.7846],[274,-10.7786],[276,-10.6543],[278,-10.5587],[280,-10.4989],[282,-10.5571],[284,-10.5052],[286,-10.4919],[288,-10.5253],[290,-10.5217],[292,-10.5418],[296,-10.5356],[298,-10.5654],[308,-10.5168],[312,-10.5131],[314,-10.5453],[320,-10.5745],[326,-10.5426],[328,-10.5523],[330,-10.0055],[332,-8.7176],[334,-7.2756],[336,-5.9938],[338,-3.9177],[340,-2.8485],[342,-2.4401],[344,-2.2716],[346,-2.2925],[348,-2.2355],[350,-2.3077],[352,-2.2742],[354,-2.3238],[356,-2.389],[358,-2.4155],[360,-2.3864],[362,-2.4791],[364,-2.7782],[366,-2.8311],[368,-2.7518]]
       }
       pinky_01_r_ctrl.Rotation.Y {
        type=MovieSceneScriptingFloatChannel
        default=4.2451
        thinned_from=67
        keys [[0,-1.0914],[222,-1.0914],[244,-10.5419],[246,-10.8284],[248,-11.3721],[250,-11.6231],[252,-12.1008],[254,-12.3195],[256,-12.5121],[258,-12.7445],[260,-12.6733],[262,-12.9187],[264,-13.0336],[266,-13.4257],[268,-13.1139],[270,-13.2053],[272,-13.5087],[276,-13.9271],[278,-13.9896],[280,-14.1542],[282,-14.2757],[284,-14.2448],[288,-14.1213],[290,-13.9677],[294,-13.9947],[296,-13.9443],[298,-14.1128],[300,-14.0038],[302,-14.0019],[304,-14.034],[312,-13.9298],[314,-13.964],[316,-13.9492],[318,-14.0949],[322,-14.118],[324,-14.1952],[326,-14.1357],[328,-14.2541],[330,-13.9086],[332,-13.0782],[334,-10.824],[336,-6.2336],[338,-2.7591],[340,-1.023],[342,-0.0227],[344,1.0258],[346,2.0067],[348,3.3175],[350,3.813],[352,4.9573],[354,4.8997],[356,4.9555],[358,5.0885],[360,4.9079],[362,4.9413],[364,4.5409],[366,4.4652],[368,4.2451]]
       }
       pinky_01_r_ctrl.Rotation.Z {
        type=MovieSceneScriptingFloatChannel
        default=-75.735
        thinned_from=67
        keys [[0,-84.6646],[222,-84.6646],[244,-47.0184],[246,-49.1182],[248,-50.2547],[250,-51.1445],[252,-51.5031],[254,-51.2141],[256,-51.4281],[258,-51.9491],[260,-51.9943],[264,-52.2768],[266,-51.9193],[268,-51.9021],[270,-52.2733],[272,-52.9947],[274,-53.4284],[276,-54.8632],[278,-56.0295],[282,-56.9475],[286,-57.371],[288,-57.3735],[292,-57.1572],[300,-57.3894],[304,-57.4498],[308,-57.9083],[310,-57.755],[320,-57.3345],[322,-57.384],[324,-57.5758],[326,-57.4464],[328,-57.7127],[330,-55.5642],[332,-49.9611],[334,-41.0197],[336,-28.9132],[338,-25.2217],[340,-24.1072],[342,-31.0003],[344,-40.6184],[346,-52.7588],[348,-64.604],[350,-74.642],[352,-82.5818],[354,-85.0988],[356,-84.7678],[358,-84.6124],[360,-84.1469],[362,-82.561],[364,-74.2997],[366,-74.2013],[368,-75.735]]
       }
       pinky_01_r_ctrl.Scale.X {type=MovieSceneScriptingFloatChannel, default=1, const [1,67]}
       pinky_01_r_ctrl.Scale.Y {type=MovieSceneScriptingFloatChannel, default=1, const [1,67]}
       pinky_01_r_ctrl.Scale.Z {type=MovieSceneScriptingFloatChannel, default=1, const [1,67]}
       pinky_02_r_ctrl.Location.X {
        type=MovieSceneScriptingFloatChannel
        default=0
        thinned_from=67
        keys [[0,-0.0001],[270,-0.0001],[272,-0.0002],[274,-0.0001],[276,-0.0002],[278,-0.0001],[342,-0.0001],[360,0],[362,-0.0004],[368,0]]
       }
       pinky_02_r_ctrl.Location.Y {
        type=MovieSceneScriptingFloatChannel
        default=0
        thinned_from=67
        keys [[0,0.0001],[222,0.0001],[244,-0.0005],[246,-0.0007],[248,-0.0003],[254,0],[270,0.0001],[272,-0.0004],[274,0.0001],[276,-0.0004],[278,0],[280,-0.0001],[332,0.0001],[336,0.0005],[338,0.0005],[342,-0.0001],[344,0.0005],[360,0],[362,-0.0004],[368,0]]
       }
       pinky_02_r_ctrl.Location.Z {
        type=MovieSceneScriptingFloatChannel
        default=-0.0001
        thinned_from=67
        keys [[0,0],[222,0],[246,-0.0002],[248,0.0001],[270,0],[272,0.0005],[274,0],[276,0.0003],[278,0],[340,0],[342,-0.0001],[368,-0.0001]]
       }
       pinky_02_r_ctrl.Rotation.X {
        type=MovieSceneScriptingFloatChannel
        default=0.3096
        thinned_from=67
        keys [[0,0],[222,0],[244,-10.4766],[246,-10.2804],[248,-10.371],[250,-10.3432],[254,-10.2043],[256,-10.0323],[258,-9.9137],[264,-9.7408],[266,-9.8045],[268,-9.6888],[270,-9.7262],[272,-9.8442],[274,-9.8465],[276,-9.8793],[278,-9.9423],[280,-9.9049],[282,-9.9749],[284,-9.9546],[286,-9.9116],[288,-9.9103],[292,-9.9755],[296,-9.9386],[298,-10.0376],[302,-9.9665],[304,-10.0047],[308,-9.9759],[310,-9.9784],[312,-9.9453],[314,-9.9757],[316,-9.9853],[318,-10.0813],[320,-10.0968],[324,-10.084],[326,-10.1301],[328,-10.1113],[330,-9.1388],[332,-6.8281],[334,-4.1035],[336,-1.6094],[338,0.1884],[340,0.3339],[342,0.1431],[344,0.3666],[346,0.5534],[348,0.8752],[350,1.2629],[352,1.268],[354,1.3774],[356,1.1942],[358,1.0569],[360,1.0188],[362,0.9357],[364,0.4719],[366,0.3833],[368,0.3096]]
       }
       pinky_02_r_ctrl.Rotation.Y {
        type=MovieSceneScriptingFloatChannel
        default=1.0041
        thinned_from=67
        keys [[0,0],[222,0],[244,2.4598],[246,2.5111],[248,2.4904],[250,2.4978],[254,2.5333],[256,2.574],[258,2.6011],[264,2.6377],[266,2.6245],[268,2.6476],[270,2.6404],[272,2.6161],[274,2.6157],[278,2.595],[280,2.6035],[282,2.5877],[286,2.6017],[288,2.602],[292,2.5875],[296,2.5958],[298,2.5734],[302,2.5896],[304,2.5809],[310,2.5869],[312,2.5943],[316,2.5854],[318,2.5633],[320,2.5596],[324,2.5626],[326,2.5516],[328,2.5561],[330,2.4913],[332,2.2937],[334,1.9565],[336,1.4936],[338,0.9834],[340,0.2928],[342,0.0104],[344,0.2109],[346,0.522],[348,1.0203],[350,1.583],[352,1.7944],[354,1.9643],[356,1.8177],[358,1.7029],[360,1.6701],[362,1.5988],[364,1.1614],[366,1.0781],[368,1.0041]]
       }
       pinky_02_r_ctrl.Rotation.Z {
        type=MovieSceneScriptingFloatChannel
        default=-58.5477
        thinned_from=67
        keys [[0,-63.0936],[222,-63.0936],[244,-15.4295],[246,-13.037],[248,-12.3125],[250,-11.0338],[252,-9.6677],[254,-8.3911],[256,-6.7305],[258,-5.5447],[260,-4.8208],[262,-4.418],[264,-4.1447],[266,-4.4947],[268,-3.8582],[270,-4.0641],[272,-4.7132],[274,-4.7257],[276,-4.9077],[278,-5.254],[280,-5.0474],[282,-5.4342],[284,-5.322],[286,-5.0849],[288,-5.0775],[290,-5.3039],[292,-5.4373],[294,-5.3041],[296,-5.2337],[298,-5.7804],[300,-5.5537],[302,-5.3876],[304,-5.5985],[306,-5.4965],[308,-5.4395],[310,-5.4532],[312,-5.2706],[314,-5.4382],[316,-5.4912],[318,-6.0219],[320,-6.1078],[322,-6.0585],[324,-6.0372],[326,-6.2926],[328,-6.1883],[330,-4.8184],[332,-2.1655],[334,-0.4923],[336,-0.5756],[338,-2.2057],[340,-8.0839],[342,-14.1231],[344,-20.5755],[346,-28.9727],[348,-36.7254],[352,-49.1895],[354,-50.857],[356,-52.1145],[358,-53.0725],[360,-53.3428],[362,-53.9299],[364,-57.3385],[368,-58.5477]]
       }
       pinky_02_r_ctrl.Scale.X {type=MovieSceneScriptingFloatChannel, default=1, const [1,67]}
       pinky_02_r_ctrl.Scale.Y {type=MovieSceneScriptingFloatChannel, default=1, const [1,67]}
       pinky_02_r_ctrl.Scale.Z {type=MovieSceneScriptingFloatChannel, default=1, const [1,67]}
       pinky_03_r_ctrl.Location.X {
        type=MovieSceneScriptingFloatChannel
        default=0.0006
        thinned_from=67
        keys [[0,0.0002],[222,0.0002],[246,-0.0003],[248,0.0001],[270,0.0002],[272,-0.0003],[274,0.0002],[276,0.0001],[332,0.0003],[338,0.0005],[342,0.0001],[344,0.0006],[360,0.0006],[362,0.0004],[368,0.0006]]
       }
       pinky_03_r_ctrl.Location.Y {
        type=MovieSceneScriptingFloatChannel
        default=-0.0002
        thinned_from=67
        keys [[0,-0.0002],[222,-0.0002],[246,0.0001],[270,-0.0002],[272,0],[274,-0.0002],[360,-0.0002],[362,-0.0003],[368,-0.0002]]
       }
       pinky_03_r_ctrl.Location.Z {
        type=MovieSceneScriptingFloatChannel
        default=0
        thinned_from=67
        keys [[0,0],[222,0],[246,-0.0005],[248,-0.0001],[270,0],[272,0.0003],[274,0],[276,0.0002],[278,0],[332,0.0001],[336,0.0003],[338,0.0003],[342,-0.0001],[344,0.0003],[360,0],[362,-0.0002],[364,0.0001],[368,0]]
       }
       pinky_03_r_ctrl.Rotation.X {
        type=MovieSceneScriptingFloatChannel
        default=-0.3087
        thinned_from=67
        keys [[0,0],[222,0],[244,-4.281],[246,-4.0674],[248,-3.7587],[250,-3.6365],[252,-3.5059],[254,-3.4243],[256,-3.3001],[258,-3.1936],[260,-3.1591],[262,-3.1054],[264,-3.0357],[266,-3.0077],[268,-2.9546],[270,-2.942],[272,-2.969],[274,-2.8603],[276,-2.8234],[278,-2.8457],[280,-2.797],[282,-2.8164],[284,-2.8202],[286,-2.7827],[288,-2.801],[290,-2.8266],[292,-2.8347],[294,-2.8207],[296,-2.8174],[298,-2.8456],[300,-2.8441],[302,-2.8288],[304,-2.8544],[306,-2.8408],[310,-2.8581],[314,-2.8564],[318,-2.8799],[320,-2.8986],[322,-2.8812],[324,-2.8767],[326,-2.8806],[328,-2.8735],[330,-2.4489],[332,-1.467],[334,-0.4797],[336,0.049],[338,0.305],[340,0.0675],[342,-0.1399],[344,-0.0512],[346,0.0932],[348,0.2673],[350,0.5821],[352,0.4975],[354,0.5965],[356,0.4663],[358,0.339],[360,0.3226],[362,0.2389],[364,-0.1921],[366,-0.2736],[368,-0.3087]]
       }
       pinky_03_r_ctrl.Rotation.Y {
        type=MovieSceneScriptingFloatChannel
        default=0.4348
        thinned_from=67
        keys [[0,0],[222,0],[244,1.8049],[246,1.7203],[248,1.5888],[250,1.5333],[252,1.472],[254,1.4323],[256,1.3699],[258,1.315],[260,1.2968],[262,1.2682],[264,1.2302],[266,1.2149],[268,1.1855],[270,1.1784],[272,1.194],[274,1.132],[276,1.1104],[278,1.1236],[280,1.0957],[282,1.1066],[284,1.1088],[286,1.087],[288,1.0977],[290,1.1126],[292,1.1173],[294,1.1092],[296,1.1072],[298,1.1235],[300,1.1226],[302,1.1139],[304,1.1286],[306,1.1208],[310,1.1307],[312,1.1289],[314,1.1297],[318,1.1432],[320,1.1539],[322,1.1439],[324,1.1414],[326,1.1436],[328,1.1396],[330,1.0663],[332,0.928],[334,0.9045],[336,1.0469],[338,1.1083],[340,1.1243],[342,1.0425],[344,0.8998],[346,0.7508],[348,0.611],[350,0.4795],[352,0.4214],[354,0.3874],[356,0.3993],[358,0.4092],[360,0.4105],[362,0.4161],[364,0.4319],[366,0.4343],[368,0.4348]]
       }
       pinky_03_r_ctrl.Rotation.Z {
        type=MovieSceneScriptingFloatChannel
        default=-56.2503
        thinned_from=67
        keys [[0,-117.2954],[222,-117.2954],[244,-6.56],[246,-3.9538],[248,-0.8059],[250,1.0843],[252,2.9138],[254,4.2993],[256,5.7978],[258,7.0138],[260,7.5681],[264,8.4826],[266,8.6643],[268,9.0105],[270,9.0925],[272,8.9283],[274,9.6275],[276,9.8688],[278,9.7239],[280,10.0453],[282,9.9161],[284,9.8909],[286,10.1381],[288,10.0174],[290,9.8487],[292,9.7954],[294,9.8876],[296,9.9098],[298,9.724],[300,9.7341],[302,9.8343],[304,9.6661],[306,9.7556],[310,9.6422],[312,9.6632],[314,9.6534],[318,9.4987],[320,9.3766],[322,9.4905],[324,9.5202],[326,9.4945],[328,9.5407],[330,11.312],[332,15.1262],[334,18.0689],[336,17.8995],[338,16.4912],[340,12.8973],[342,6.7349],[344,-2.8027],[346,-14.9946],[348,-27.836],[350,-38.562],[352,-48.1335],[354,-51.0683],[358,-52.545],[360,-52.6388],[362,-53.1172],[364,-55.5842],[366,-56.0491],[368,-56.2503]]
       }
       pinky_03_r_ctrl.Scale.X {type=MovieSceneScriptingFloatChannel, default=1, const [1,67]}
       pinky_03_r_ctrl.Scale.Y {type=MovieSceneScriptingFloatChannel, default=1, const [1,67]}
       pinky_03_r_ctrl.Scale.Z {type=MovieSceneScriptingFloatChannel, default=1, const [1,67]}
       hand_r_ik_ctrl.Location.X {
        type=MovieSceneScriptingFloatChannel
        default=-22.6349
        thinned_from=131
        keys [[0,-410.0948],[94,-404.1292],[96,-370.8132],[98,-327.8688],[100,-279.3224],[102,-228.5783],[104,-180.9463],[106,-140.7671],[108,-108.991],[110,-92.079],[112,-82.2421],[114,-74.3896],[116,-68.4657],[118,-64.8701],[120,-62.6134],[122,-60.8597],[124,-58.6765],[128,-53.9304],[134,-47.4716],[138,-42.9804],[140,-40.8362],[142,-38.8315],[144,-37.041],[148,-33.8576],[150,-32.3286],[154,-29.4611],[156,-28.0829],[158,-26.8416],[160,-25.8661],[162,-24.9678],[164,-24.1338],[168,-23.0645],[170,-22.5883],[188,-17.7153],[190,-17.0617],[192,-16.122],[212,-15.2686],[214,-14.5518],[216,-14.0381],[218,-14.0012],[220,-14.214],[222,-14.797],[224,-15.6379],[226,-16.4114],[228,-17.8557],[230,-19.0057],[232,-20.8962],[234,-22.4403],[236,-24.4279],[238,-27.3317],[240,-31.1564],[242,-35.5728],[244,-40.2531],[246,-44.5605],[248,-49.5062],[250,-52.1576],[252,-54.1341],[254,-55.3971],[256,-56.3996],[258,-57.0674],[264,-58.3113],[266,-60.5341],[270,-61.185],[274,-61.5431],[278,-61.3353],[282,-61.3639],[292,-60.2818],[312,-58.9344],[322,-58.6093],[326,-58.2814],[328,-57.9894],[330,-57.4347],[332,-56.3206],[334,-53.9222],[336,-47.6416],[338,-39.3053],[340,-29.1372],[342,-18.652],[344,-12.6503],[346,-8.0021],[348,-5.1079],[350,-2.5315],[352,-0.1068],[354,1.1356],[356,1.5401],[358,0.7957],[360,-0.5278],[362,-3.6165],[364,-15.036],[366,-30.1619],[368,-22.6349]]
       }
       hand_r_ik_ctrl.Location.Y {
        type=MovieSceneScriptingFloatChannel
        default=-61.3361
        thinned_from=131
        keys [[0,20.7549],[92,16.9738],[94,17.535],[96,17.223],[98,16.9762],[100,16.775],[102,16.7413],[104,15.9337],[106,14.1632],[108,12.9554],[110,10.7517],[112,10.5152],[114,11.7163],[116,13.9083],[120,18.7241],[122,20.5754],[124,21.7811],[130,22.7938],[132,23.0471],[136,23.4088],[138,23.4173],[142,23.1026],[144,22.7189],[146,22.0853],[148,21.5031],[150,20.8747],[152,20.141],[156,18.8979],[158,18.3855],[160,17.9275],[162,17.5227],[164,17.2552],[166,17.128],[168,17.2081],[170,17.4187],[174,18.0092],[176,18.2648],[180,18.3511],[188,18.7141],[192,18.222],[212,17.5745],[214,16.6276],[216,15.4657],[218,14.2422],[220,12.8843],[224,10.0849],[226,8.8157],[228,7.8191],[230,6.9665],[232,6.5446],[234,5.4277],[236,5.1685],[238,5.381],[240,5.7945],[242,6.4821],[244,7.3967],[246,8.6705],[248,10.1886],[250,11.3173],[252,12.0879],[254,12.5789],[256,13.1222],[258,13.5176],[260,13.6645],[262,13.942],[264,14.1861],[266,14.0211],[268,13.9137],[270,13.6803],[272,12.8928],[274,11.9251],[276,10.6689],[278,9.8647],[280,9.4769],[282,9.2988],[284,9.2698],[286,9.3619],[288,9.4077],[290,9.4784],[292,9.5077],[294,9.5612],[300,9.6485],[302,9.6944],[304,9.7619],[312,10.18],[314,10.2713],[318,10.3453],[326,10.4246],[328,10.564],[330,10.7737],[332,10.9104],[334,11.1134],[336,12.1219],[338,12.9451],[340,13.1748],[342,13.6212],[344,12.8126],[346,12.2831],[348,13.0234],[350,11.9631],[352,11.6736],[354,12.7203],[356,14.2278],[358,16.6077],[360,17.1101],[362,16.5576],[364,-0.8847],[366,-26.0938],[368,-61.3361]]
       }
       hand_r_ik_ctrl.Location.Z {
        type=MovieSceneScriptingFloatChannel
        default=237.5941
        thinned_from=131
        keys [[0,-805.8495],[94,-815.6234],[96,-750.2848],[98,-666.7435],[100,-572.3215],[102,-475.0156],[104,-380.7483],[106,-295.8851],[108,-230.8923],[110,-190.143],[112,-168.6548],[114,-150.6628],[116,-136.1271],[118,-123.3242],[120,-112.0177],[122,-101.3177],[126,-82.6549],[128,-74.6094],[130,-67.8136],[132,-61.6247],[134,-56.1293],[136,-50.9821],[138,-46.1041],[140,-41.4928],[142,-37.2859],[144,-33.4277],[146,-30.1503],[148,-27.2946],[150,-24.8001],[152,-22.7919],[154,-20.8664],[156,-19.6135],[158,-18.7953],[160,-18.147],[162,-17.6511],[164,-17.3042],[166,-17.3079],[168,-17.5698],[170,-17.9772],[172,-18.3484],[174,-18.432],[178,-18.2469],[180,-18.0699],[182,-17.8419],[184,-17.5662],[186,-17.1664],[188,-16.7033],[190,-15.7663],[192,-13.9846],[212,-11.7824],[214,-9.0735],[216,-5.9617],[218,-2.3872],[220,1.4719],[222,5.3969],[224,9.2921],[226,12.5301],[228,16.2622],[230,19.5607],[232,23.3708],[234,26.3275],[236,29.8249],[238,33.4302],[240,36.5111],[242,38.7966],[244,39.9571],[246,40.313],[248,39.3543],[250,38.6372],[252,37.7218],[254,37.0284],[256,36.4352],[258,36.0923],[264,35.6963],[266,36.4258],[268,36.2281],[270,36.1164],[272,36.7683],[274,37.7111],[276,39.0606],[278,39.6874],[280,39.6833],[282,39.5213],[284,39.2146],[286,39.0312],[300,39.1716],[310,38.8595],[326,38.9114],[328,39.0162],[332,39.6971],[334,39.3844],[336,35.8665],[338,31.3823],[340,25.2254],[342,16.0588],[344,8.0022],[346,0.757],[348,-4.2246],[350,-4.9265],[352,-4.0689],[354,-1.847],[356,0.972],[358,4.9331],[360,9.7897],[362,15.309],[364,34.9467],[366,105.1914],[368,237.5941]]
       }
       hand_r_ik_ctrl.Rotation.X {
        type=MovieSceneScriptingFloatChannel
        default=-50.5424
        thinned_from=131
        keys [[0,-17.4929],[92,5.0524],[94,6.9046],[96,10.1272],[98,15.7025],[100,22.9483],[102,31.422],[104,38.0508],[106,42.1243],[108,46.3362],[110,48.1235],[112,51.3859],[114,53.8272],[116,54.0577],[118,52.0171],[120,48.5156],[122,44.8288],[124,41.4836],[126,39.3272],[128,37.2733],[130,35.5528],[132,34.3161],[134,33.7085],[136,33.631],[138,33.0111],[140,32.5802],[142,32.9712],[144,32.9656],[146,31.1673],[148,28.6678],[150,26.625],[152,24.744],[154,23.6288],[156,21.3035],[158,19.4609],[160,17.795],[162,16.4074],[164,15.1027],[166,13.5831],[168,12.7935],[170,11.9155],[172,11.3089],[174,10.9962],[176,10.8752],[178,10.9696],[180,11.1618],[182,11.2537],[184,11.2489],[186,11.2211],[188,11.1674],[190,11.0193],[192,10.9252],[212,9.5963],[214,6.5618],[216,1.1694],[218,-10.6191],[220,-17.4413],[222,-18.521],[224,-31.5962],[226,-31.9733],[228,-34.5948],[230,-20.226],[232,-15.0982],[234,-15.2819],[236,-17.1141],[238,-21.5119],[240,-38.9236],[242,-67.484],[244,-94.9247],[246,-109.8239],[248,-120.4679],[250,-126.5],[252,-130.4051],[256,-133.3831],[258,-134.2424],[262,-134.5997],[264,-135.0338],[266,-135.8677],[268,-134.1797],[272,-131.7379],[274,-130.2884],[276,-126.4811],[280,-116.1143],[282,-112.0707],[284,-109.415],[286,-108.9328],[290,-109.5878],[294,-110.7285],[302,-111.4281],[308,-111.5508],[326,-111.5606],[330,-112.3778],[332,-113.2614],[334,-115.1664],[336,-115.3763],[338,-102.4175],[340,-72.9704],[342,-20.5648],[344,13.6773],[346,27.4006],[348,20.9687],[350,19.2687],[352,20.2877],[354,20.8112],[356,25.1231],[358,19.1269],[360,14.6815],[362,-5.3478],[364,-45.869],[366,-52.6295],[368,-50.5424]]
       }
       hand_r_ik_ctrl.Rotation.Y {
        type=MovieSceneScriptingFloatChannel
        default=53.7149
        thinned_from=131
        keys [[0,56.7455],[92,51.5425],[96,52.7769],[98,53.0919],[100,52.6987],[102,50.9287],[104,48.7228],[106,47.2019],[108,44.2151],[110,42.4789],[112,37.3431],[114,29.4913],[116,19.8196],[118,11.3579],[120,5.1087],[122,2.3304],[124,1.4267],[128,5.2893],[130,6.5617],[132,8.9468],[134,11.4448],[136,14.6747],[138,18.2644],[142,25.3578],[144,28.3665],[146,29.8311],[148,30.7497],[150,31.3445],[152,31.11],[154,30.3415],[156,28.5964],[158,26.9447],[160,25.389],[164,21.9301],[166,19.9716],[168,17.8219],[170,15.5593],[172,13.3585],[174,11.5473],[176,10.0848],[178,9.2205],[180,8.4084],[182,7.7401],[184,7.2578],[186,7.032],[188,7.0262],[190,8.3054],[192,11.3751],[212,14.539],[214,18.6935],[216,24.652],[218,31.1755],[220,38.052],[222,45.8333],[224,49.5651],[226,63.3732],[228,70.9529],[230,65.6517],[232,69.636],[234,65.6142],[236,62.96],[238,60.6616],[240,57.5323],[242,56.6283],[244,49.2981],[246,40.6209],[248,26.2075],[250,19.5741],[252,14.8086],[254,12.1563],[256,9.7693],[258,8.1714],[260,7.6222],[264,6.1643],[266,5.5434],[268,4.0747],[270,2.8138],[272,3.1466],[274,4.8631],[276,7.7902],[278,12.4888],[280,18.2326],[282,22.6861],[284,24.8509],[286,25.1851],[288,24.8477],[290,24.051],[292,23.0345],[294,22.1008],[298,20.577],[300,20.1583],[304,19.9729],[306,19.912],[308,19.9132],[310,20.0185],[314,20.0662],[316,20.2068],[318,20.2199],[322,20.5924],[326,21.1965],[328,21.8585],[330,23.9863],[332,26.72],[334,30.1387],[336,36.6522],[338,42.2546],[340,53.5656],[342,59.1175],[344,50.4471],[346,12.9531],[348,31.5604],[350,30.5117],[352,31.6438],[356,40.5266],[358,51.3767],[360,58.9769],[362,54.5315],[364,77.2969],[366,51.2724],[368,53.7149]]
       }
       hand_r_ik_ctrl.Rotation.Z {
        type=MovieSceneScriptingFloatChannel
        default=-134.3373
        thinned_from=131
        keys [[0,-81.2257],[92,-68.0128],[94,-64.9626],[96,-60.4541],[98,-53.6689],[100,-45.2711],[102,-35.8187],[104,-28.332],[108,-18.4165],[110,-16.0799],[112,-10.2534],[114,-1.603],[116,7.7985],[118,16.7856],[120,24.8406],[122,30.6269],[124,33.1412],[126,32.8038],[128,32.3762],[130,32.1674],[134,30.9418],[136,30.1896],[138,28.9119],[140,27.3997],[142,26.4086],[144,24.6852],[146,21.6319],[148,18.508],[150,15.1921],[152,11.4573],[154,7.8921],[156,4.9701],[158,2.3033],[160,-0.314],[162,-2.6633],[164,-4.904],[166,-7.1231],[168,-9.7897],[170,-11.4629],[172,-11.9952],[174,-13.3066],[176,-14.475],[178,-15.4098],[180,-16.1875],[182,-16.7925],[184,-17.1985],[186,-17.5209],[188,-17.6684],[190,-18.4786],[192,-19.1064],[212,-20.2431],[214,-22.6263],[216,-24.7405],[218,-25.7746],[220,-30.2715],[222,-28.3729],[224,-35.8239],[226,-38.4331],[228,-40.0669],[230,-27.4217],[232,-24.3006],[234,-29.2848],[236,-38.4541],[238,-52.5991],[240,-76.8901],[242,-94.183],[244,-118.2176],[246,-133.53],[248,-145.7263],[250,-150.5039],[252,-154.0657],[254,-156.8786],[256,-159.0183],[258,-160.0544],[260,-160.4221],[264,-161.6198],[266,-161.6658],[268,-163.4276],[270,-163.6057],[272,-161.0234],[274,-157.071],[276,-151.9807],[278,-148.683],[280,-147.3405],[282,-146.2931],[284,-146.6193],[286,-147.9626],[294,-152.0553],[296,-152.7607],[302,-153.5968],[326,-155.3741],[328,-156.6157],[330,-158.8903],[332,-159.5553],[334,-156.9481],[336,-149.3046],[338,-134.5317],[340,-112.9873],[342,-64.8981],[344,-29.441],[346,-8.4056],[348,-7.3661],[350,-8.9527],[352,-6.3665],[354,-6.1704],[356,6.1974],[358,-11.3487],[360,-17.6314],[362,-4.4661],[364,-76.0447],[366,-132.8306],[368,-134.3373]]
       }
       hand_r_ik_ctrl.Scale.X {type=MovieSceneScriptingFloatChannel, default=1, const [1,131]}
       hand_r_ik_ctrl.Scale.Y {type=MovieSceneScriptingFloatChannel, default=1, const [1,131]}
       hand_r_ik_ctrl.Scale.Z {type=MovieSceneScriptingFloatChannel, default=1, const [1,131]}
       Softness_2 {type=MovieSceneScriptingFloatChannel, default=0, const [0,131]}
       hand_r_ikorient_ctrl.Rotation.Roll {
        type=MovieSceneScriptingFloatChannel
        default=0
        thinned_from=140
        keys [[0,0],[344,0],[346,29.6691],[348,0],[350,28.1688],[352,0],[354,0],[356,28.5681],[358,0],[360,30.0636],[362,0],[368,0]]
       }
       hand_r_ikorient_ctrl.Rotation.Pitch {
        type=MovieSceneScriptingFloatChannel
        default=0
        thinned_from=140
        keys [[0,0],[344,0],[346,10.7647],[348,0],[350,-16.8861],[352,0],[354,0],[356,-15.4637],[358,0],[360,-18.3882],[362,0],[368,0]]
       }
       hand_r_ikorient_ctrl.Rotation.Yaw {
        type=MovieSceneScriptingFloatChannel
        default=0
        thinned_from=140
        keys [[0,0],[344,0],[346,27.1908],[348,0],[350,26.6331],[352,0],[354,0],[356,18.6621],[358,0],[360,22.6546],[362,0],[368,0]]
       }
       arm_r_pv_ik_ctrl.Location.X {
        type=MovieSceneScriptingFloatChannel
        default=33.056
        thinned_from=131
        keys [[0,4.6994],[92,8.4972],[98,9.08],[102,9.3944],[104,9.733],[106,10.1504],[108,10.5096],[110,10.9683],[112,10.9409],[114,10.1514],[116,8.4545],[118,5.7705],[120,2.2908],[122,-1.3098],[124,-4.222],[126,-6.228],[128,-7.7812],[130,-8.8041],[132,-9.5777],[134,-9.9728],[136,-10.2006],[138,-10.5554],[140,-10.9482],[142,-11.3801],[144,-12.0487],[146,-13.0039],[150,-15.0602],[152,-16.1257],[154,-17.0325],[156,-17.8244],[158,-18.7523],[160,-19.7998],[162,-20.7419],[164,-21.6422],[166,-22.6272],[168,-23.8575],[172,-26.528],[174,-27.4433],[176,-28.1628],[178,-28.4806],[184,-29.1971],[186,-29.3318],[188,-29.3878],[190,-28.836],[192,-27.9485],[212,-27.0267],[214,-26.2081],[216,-25.4862],[222,-23.817],[224,-23.2241],[228,-22.2337],[230,-21.7913],[232,-21.4457],[234,-21.7634],[236,-21.8139],[238,-21.782],[240,-21.8676],[242,-22.46],[244,-22.9956],[246,-23.2423],[248,-23.1879],[250,-23.1821],[252,-23.0397],[256,-22.5194],[258,-22.3286],[264,-22.1495],[266,-22.0043],[272,-20.8176],[274,-20.6415],[276,-20.6472],[278,-20.5012],[280,-20.2836],[282,-20.2779],[286,-20.8029],[290,-21.1689],[294,-21.3051],[302,-21.3531],[304,-21.3379],[310,-21.3978],[318,-21.1063],[320,-20.9503],[322,-20.8746],[326,-20.4553],[328,-19.6439],[330,-17.8612],[332,-15.3705],[334,-12.2729],[336,-9.6554],[338,-9.0135],[340,-10.7303],[342,-13.9373],[344,-16.2435],[346,-18.8045],[348,-20.68],[350,-19.731],[352,-18.1785],[354,-17.3822],[356,-17.0048],[358,-17.384],[360,-19.0788],[362,-21.4837],[364,-5.2804],[366,20.4461],[368,33.056]]
       }
       arm_r_pv_ik_ctrl.Location.Y {
        type=MovieSceneScriptingFloatChannel
        default=217.7513
        thinned_from=131
        keys [[0,-894.6631],[94,-904.0709],[96,-831.1407],[98,-737.6616],[100,-631.889],[102,-522.4343],[104,-417.003],[106,-323.3909],[108,-251.2362],[110,-207.7215],[112,-184.4472],[114,-165.3281],[116,-150.2467],[118,-138.1079],[120,-128.0621],[122,-118.8768],[126,-101.3178],[128,-93.612],[130,-86.6957],[132,-80.6452],[134,-75.144],[136,-70.0618],[138,-65.2965],[140,-60.7061],[142,-56.3596],[144,-52.4531],[146,-49.1607],[148,-46.1396],[150,-43.3739],[152,-40.9572],[154,-38.4298],[156,-36.3817],[158,-34.6726],[160,-33.2046],[162,-31.8195],[164,-30.5436],[166,-29.7056],[170,-28.2636],[172,-27.4724],[176,-25.6669],[180,-24.0018],[182,-23.2284],[186,-21.7842],[188,-21.1155],[190,-20.352],[192,-19.2454],[212,-18.0528],[214,-16.7086],[216,-15.1962],[218,-13.5843],[220,-11.6372],[222,-9.3681],[224,-6.7034],[226,-3.7247],[228,-0.3461],[230,3.4403],[232,7.5318],[236,16.8765],[238,21.2355],[240,25.0658],[242,28.641],[244,31.6423],[246,34.1857],[248,36.8095],[250,38.3548],[252,39.39],[254,39.9249],[256,40.2651],[258,40.4365],[264,40.5337],[270,41.0688],[280,41.3138],[282,41.2501],[286,40.8892],[290,40.7914],[302,41.0029],[314,40.7213],[322,40.759],[326,40.6424],[328,40.3481],[330,39.5609],[332,37.7639],[334,33.4709],[336,22.82],[338,10.617],[340,-0.3618],[342,-7.6908],[344,-10.4763],[346,-11.2874],[348,-11.3509],[350,-10.3455],[352,-8.7311],[354,-7.5301],[356,-6.461],[358,-4.9198],[360,-2.9604],[362,0.1468],[364,17.9713],[366,87.2405],[368,217.7513]]
       }
       arm_r_pv_ik_ctrl.Location.Z {
        type=MovieSceneScriptingFloatChannel
        default=116.914
        thinned_from=131
        keys [[0,-8.0323],[92,-9.2602],[94,-9.9502],[96,-10.6746],[98,-11.5318],[100,-12.2118],[102,-13.0418],[104,-12.4802],[106,-9.8902],[108,-7.9583],[110,-3.9189],[112,-1.3715],[114,1.0521],[116,3.419],[118,6.3658],[120,9.4811],[122,12.3924],[124,14.8614],[126,16.2782],[128,17.0965],[130,17.8222],[132,17.6017],[134,17.0476],[136,15.845],[138,14.4847],[140,13.268],[142,12.2918],[144,11.5334],[148,10.6772],[150,10.2838],[152,10.027],[154,9.7217],[156,9.5525],[158,9.4935],[160,9.4613],[164,9.5741],[166,9.8592],[168,10.0056],[170,9.997],[172,9.7229],[174,9.5225],[176,9.2913],[178,9.1005],[180,8.8865],[182,8.6217],[186,7.9625],[188,7.617],[190,7.2472],[192,6.5424],[212,5.8363],[214,5.029],[216,4.0806],[218,3.1064],[220,1.8795],[222,0.5335],[224,-0.8256],[226,-2.2693],[228,-3.6136],[230,-4.665],[232,-5.355],[234,-5.8316],[236,-5.6551],[238,-4.9756],[240,-4.0298],[242,-3.0034],[244,-1.9435],[246,-0.7721],[248,0.7598],[250,1.9403],[252,2.9628],[254,3.7521],[256,4.4525],[258,4.9867],[260,5.381],[262,5.7615],[264,6.1049],[266,8.676],[270,9.5175],[272,10.0346],[274,10.4738],[276,10.8426],[278,11.114],[280,11.2278],[282,11.1326],[284,10.6999],[286,10.3312],[288,10.0527],[290,9.8374],[292,9.6791],[296,9.4729],[300,9.2034],[306,8.6466],[308,8.4289],[310,8.3055],[312,8.2065],[316,8.1008],[320,8.0421],[324,7.919],[326,7.8098],[328,7.7815],[330,7.9365],[332,8.0743],[334,7.7172],[336,5.1133],[338,4.2469],[340,5.5298],[342,7.2547],[344,7.4392],[346,4.8899],[348,2.0564],[350,-0.7287],[352,-3.3539],[354,-4.9412],[356,-6.2545],[358,-7.5344],[360,-10.0181],[362,-13.6176],[364,-7.0807],[366,48.1695],[368,116.914]]
       }
       arm_r_pv_ik_ctrl.Rotation.X {
        type=MovieSceneScriptingFloatChannel
        default=-85.7678
        thinned_from=131
        keys [[0,-45.3403],[92,-25.8953],[94,-24.7275],[96,-22.6998],[98,-19.1667],[100,-14.7172],[102,-9.348],[104,-4.9483],[106,-2.2537],[108,1.3619],[110,3.3252],[112,7.0139],[114,10.7569],[116,14.8972],[118,18.6675],[120,22.3986],[122,25.2094],[124,27.7809],[126,27.7831],[128,27.4011],[130,27.3435],[132,25.9796],[134,24.3899],[136,22.1195],[138,19.6213],[140,17.178],[142,15.0761],[144,13.0284],[146,11.4238],[148,10.1018],[150,8.9744],[152,8.1056],[154,7.2991],[156,7.0332],[158,7.1782],[160,7.2194],[162,7.5061],[164,7.921],[166,8.8383],[168,9.7131],[170,10.4401],[172,10.8181],[174,11.2364],[176,11.5662],[178,11.8286],[180,11.9994],[182,12.1027],[184,12.1398],[186,12.0569],[188,11.9386],[190,11.5047],[192,10.4159],[212,8.9119],[214,6.845],[216,4.2306],[218,1.1255],[220,-2.5122],[222,-6.6069],[224,-10.9871],[226,-15.8401],[230,-26.1475],[232,-31.1577],[234,-37.1737],[236,-42.6371],[238,-47.5343],[240,-51.8064],[242,-55.7362],[244,-59.1232],[246,-62.0879],[248,-65.5129],[250,-67.4938],[252,-68.9903],[254,-69.8865],[256,-70.4627],[258,-70.7121],[266,-70.7045],[274,-72.7036],[278,-73.4136],[280,-73.6002],[282,-73.5336],[286,-72.6106],[290,-72.0463],[306,-71.3487],[312,-70.97],[322,-70.7777],[326,-70.2354],[328,-69.6511],[330,-68.6204],[332,-66.4952],[334,-61.2229],[336,-46.6376],[338,-28.3626],[340,-9.0743],[342,7.0117],[344,12.4048],[346,13.4793],[348,13.0267],[350,11.4727],[352,9.2041],[354,8.5438],[356,8.0036],[358,5.9863],[360,3.3595],[362,-2.4888],[364,-23.88],[366,-80.3634],[368,-85.7678]]
       }
       arm_r_pv_ik_ctrl.Rotation.Y {
        type=MovieSceneScriptingFloatChannel
        default=70.5271
        thinned_from=131
        keys [[0,50.2416],[92,49.0615],[94,48.222],[96,46.9888],[98,45.0633],[100,42.5189],[102,39.1895],[104,36.3911],[106,34.8338],[108,32.4308],[110,31.652],[112,28.4112],[114,23.4792],[116,17.2468],[118,11.3245],[120,5.934],[122,1.9946],[124,-0.9282],[126,-2.0438],[128,-3.0411],[130,-4.0744],[132,-4.5493],[134,-4.6913],[136,-4.477],[138,-4.0433],[140,-3.5019],[142,-3.0036],[144,-2.1658],[146,-1.2124],[148,-0.5085],[150,0.1796],[152,0.976],[154,1.5311],[156,2.1654],[158,2.64],[160,3.2487],[162,3.8285],[164,4.2833],[166,4.7771],[168,4.9986],[170,5.1038],[172,5.0704],[176,5.0314],[178,5.2575],[180,5.4691],[182,5.5899],[184,5.6234],[186,5.6794],[188,5.6875],[190,5.9366],[192,5.9709],[212,6.4184],[214,7.1083],[216,7.8556],[218,8.53],[220,8.9781],[222,9.1049],[224,8.8741],[226,8.1074],[228,6.8664],[230,5.2206],[232,3.3045],[234,0.4889],[236,-1.6557],[238,-3.058],[240,-4.1136],[242,-5.6069],[244,-6.758],[246,-7.0443],[248,-7.0666],[250,-6.75],[252,-6.5316],[254,-6.2193],[256,-5.7128],[258,-5.2808],[260,-5.0854],[262,-4.8254],[264,-4.5418],[266,-3.8658],[268,-3.64],[270,-3.3674],[272,-3.0006],[274,-2.693],[278,-2.2057],[280,-2.0611],[282,-2.1649],[284,-2.7249],[286,-3.128],[288,-3.3232],[290,-3.4822],[292,-3.5934],[294,-3.6039],[296,-3.6039],[300,-3.7021],[302,-3.7748],[304,-3.7769],[306,-3.8127],[308,-3.8735],[310,-3.7997],[312,-3.6331],[314,-3.4968],[316,-3.4258],[318,-3.3138],[320,-3.1387],[322,-3.0605],[324,-2.8302],[326,-2.6483],[328,-1.497],[330,1.1677],[332,4.8156],[334,9.3896],[336,14.3628],[338,19.2027],[340,21.0121],[342,18.7135],[344,17.4935],[346,15.4628],[348,14.1483],[350,13.8677],[352,13.3352],[354,11.5466],[356,8.9646],[358,6.4748],[360,2.5995],[362,0.6712],[364,20.6831],[366,53.2193],[368,70.5271]]
       }
       arm_r_pv_ik_ctrl.Rotation.Z {
        type=MovieSceneScriptingFloatChannel
        default=-12.8691
        thinned_from=131
        keys [[0,21.3105],[92,28.0069],[94,28.8363],[96,30.3006],[98,32.8468],[100,35.9322],[102,39.2563],[104,41.4494],[106,42.2636],[108,42.9878],[110,42.245],[112,41.7544],[114,40.1927],[116,37.1864],[118,32.7291],[120,27.4571],[122,22.3107],[124,18.401],[126,15.9076],[128,14.1128],[130,13.2078],[132,12.7111],[134,12.7982],[136,13.1121],[138,13.0575],[140,12.8458],[142,12.5669],[144,11.7664],[146,10.1533],[148,8.3371],[150,6.3885],[152,4.2178],[154,2.3953],[156,0.5625],[158,-1.4323],[160,-3.4959],[162,-5.3889],[166,-8.8475],[168,-10.8113],[170,-12.8868],[172,-14.9311],[174,-16.34],[176,-17.4617],[178,-18.1026],[182,-19.1109],[184,-19.5205],[186,-19.7897],[188,-19.8929],[190,-19.1973],[192,-17.9376],[212,-17.0983],[214,-16.8697],[216,-17.2015],[218,-17.8998],[220,-19.0594],[224,-21.8419],[226,-23.0031],[228,-23.7144],[230,-23.7591],[232,-22.9665],[234,-23.1453],[236,-21.2584],[238,-18.6427],[240,-16.3177],[242,-14.144],[244,-12.3042],[246,-10.2804],[248,-8.9227],[250,-7.7065],[252,-7.3117],[254,-7.3199],[256,-7.2506],[258,-7.1625],[260,-7.1163],[264,-6.9288],[266,-6.7682],[268,-7.2277],[270,-7.666],[272,-8.0876],[274,-8.2153],[276,-8.1298],[278,-8.2583],[280,-8.5412],[282,-8.6747],[284,-8.8322],[286,-8.813],[288,-8.6907],[290,-8.4779],[294,-8.1673],[298,-8.0705],[300,-8.0478],[304,-8.1363],[306,-8.2204],[308,-8.368],[316,-8.6976],[318,-8.7258],[320,-8.7906],[322,-8.8277],[324,-8.9117],[326,-8.9296],[328,-9.0198],[330,-9.4497],[332,-10.4926],[334,-12.5442],[336,-15.2204],[338,-14.7725],[340,-12.0647],[342,-10.0355],[344,-11.88],[346,-15.9043],[348,-19.8188],[350,-20.7407],[352,-20.597],[354,-18.706],[356,-16.053],[358,-13.9676],[360,-10.9463],[362,-10.1439],[364,-20.5189],[366,-30.526],[368,-12.8691]]
       }
       arm_r_pv_ik_ctrl.Scale.X {type=MovieSceneScriptingFloatChannel, default=1, const [1,131]}
       arm_r_pv_ik_ctrl.Scale.Y {type=MovieSceneScriptingFloatChannel, default=1, const [1,131]}
       arm_r_pv_ik_ctrl.Scale.Z {type=MovieSceneScriptingFloatChannel, default=1, const [1,131]}
       body_ctrl.Location.X {
        type=MovieSceneScriptingFloatChannel
        default=-8.9405
        thinned_from=140
        keys [[0,18.7997],[92,18.7997],[98,18.5104],[100,18.3717],[102,18.2814],[104,18.1268],[106,17.8532],[108,17.7023],[110,17.4242],[114,17.0515],[116,16.8292],[118,16.5596],[120,16.3477],[122,16.197],[124,16.1481],[126,16.0536],[128,16.0031],[134,15.9293],[140,16.0141],[142,16.0761],[144,16.1853],[148,16.3028],[158,16.3417],[170,16.2067],[176,16.1286],[198,16.3254],[214,16.6082],[226,17.0024],[270,17.3523],[278,17.5598],[282,17.6095],[304,17.6114],[320,17.1905],[328,17.0252],[332,16.9836],[334,16.9266],[336,16.7703],[338,16.5303],[340,16.1794],[342,15.7713],[344,15.5292],[348,14.8506],[352,13.9634],[354,13.8923],[356,13.9779],[358,14.4375],[360,15.7266],[362,17.7155],[364,16.813],[366,11.0208],[368,-8.9405]]
       }
       body_ctrl.Location.Y {
        type=MovieSceneScriptingFloatChannel
        default=147.804
        thinned_from=81
        keys [[208,0.6646],[210,0.695],[212,0.7116],[216,0.7193],[218,0.7173],[220,0.7104],[222,0.7169],[224,0.732],[226,0.7664],[228,0.8104],[230,0.8942],[232,0.951],[234,1.0153],[236,1.0682],[238,1.1138],[240,1.1429],[242,1.1578],[244,1.1792],[248,1.2067],[250,1.2174],[254,1.2314],[258,1.2498],[260,1.2566],[262,1.2575],[266,1.2363],[268,1.2016],[270,1.1739],[272,1.1391],[274,1.1016],[276,1.057],[278,1.0182],[280,0.9831],[282,0.9549],[284,0.932],[286,0.9198],[288,0.9141],[292,0.9266],[294,0.9301],[302,0.9288],[304,0.9315],[306,0.9517],[310,1.0434],[312,1.1019],[314,1.1732],[316,1.2549],[318,1.322],[320,1.3756],[322,1.3906],[324,1.4089],[326,1.4113],[328,1.408],[330,1.3793],[332,1.3031],[334,1.074],[336,0.3745],[338,-0.566],[340,-1.4802],[342,-2.1373],[344,-2.4155],[346,-2.6586],[348,-2.7736],[350,-2.8188],[352,-2.7388],[354,-2.5626],[356,-2.4249],[358,-2.3607],[360,-2.2824],[362,-2.1042],[364,6.8014],[366,43.2365],[368,147.804]]
       }
       body_ctrl.Location.Z {
        type=MovieSceneScriptingFloatChannel
        default=77.3337
        thinned_from=140
        keys [[0,-10.2341],[92,-10.2341],[94,-10.3356],[96,-10.406],[98,-10.3893],[102,-10.0702],[104,-9.6627],[106,-8.677],[108,-7.9425],[110,-6.5313],[112,-5.514],[114,-4.1576],[116,-2.6979],[118,-1.0492],[120,0.5423],[122,2.0897],[124,3.3466],[126,4.6604],[128,5.7473],[130,6.6106],[132,7.2315],[134,7.5541],[136,7.5652],[140,7.3033],[142,7.1922],[146,7.0157],[148,6.8511],[150,6.6579],[152,6.4505],[154,6.2069],[156,5.9391],[160,5.359],[162,5.0573],[164,4.729],[166,4.3868],[170,3.6707],[172,3.3031],[174,2.9454],[178,2.2665],[182,1.5741],[184,1.2339],[186,0.9374],[188,0.6542],[190,0.4026],[192,0.1823],[194,0.0029],[196,-0.1309],[198,-0.2412],[200,-0.3265],[202,-0.3893],[204,-0.4192],[206,-0.4167],[208,-0.3613],[210,-0.2706],[212,-0.1543],[214,-0.0207],[216,0.1306],[218,0.315],[220,0.5196],[222,0.7467],[224,0.9897],[226,1.2451],[228,1.5102],[230,1.7873],[232,2.0851],[234,2.4011],[238,3.0422],[242,3.6651],[246,4.2676],[248,4.5599],[250,4.8422],[252,5.1039],[254,5.351],[258,5.7999],[260,6.0121],[262,6.2037],[264,6.3711],[266,6.5119],[268,6.6316],[270,6.7328],[272,6.8109],[274,6.8463],[276,6.8456],[278,6.8261],[282,6.748],[292,6.4534],[296,6.2902],[298,6.1978],[300,6.0924],[308,5.5859],[312,5.3769],[314,5.2963],[318,5.1781],[322,5.0778],[326,5.0108],[330,4.9716],[332,4.9408],[334,4.8444],[336,4.1612],[338,2.5949],[340,0.3261],[342,-2.0203],[344,-3.5746],[346,-5.5996],[348,-7.2723],[350,-8.7188],[352,-9.6664],[354,-10.1426],[356,-10.3608],[358,-10.3163],[360,-10.9741],[362,-12.2919],[364,-8.5073],[366,16.3775],[368,77.3337]]
       }
       body_ctrl.Rotation.X {
        type=MovieSceneScriptingFloatChannel
        default=60.5079
        thinned_from=140
        keys [[0,116.2705],[92,116.2705],[94,115.7159],[98,113.5827],[102,110.6239],[104,107.3897],[106,101.6165],[108,96.4612],[110,88.8468],[112,82.6104],[114,75.7834],[116,69.3447],[118,62.6604],[120,56.7694],[122,51.3457],[124,48.0921],[126,43.9286],[128,40.2095],[130,37.6064],[132,34.4646],[136,29.1481],[138,26.7633],[140,24.4543],[142,22.3965],[144,20.1062],[146,17.9212],[148,16.0416],[150,14.3186],[152,12.6682],[154,11.2302],[156,10.0759],[158,9.0664],[160,7.9593],[162,7.146],[164,6.5554],[166,6.1004],[168,5.6719],[172,4.7683],[174,4.4366],[176,4.1229],[178,3.88],[180,3.6594],[182,3.4679],[184,3.306],[186,3.0978],[190,2.7839],[192,2.6606],[194,2.558],[196,2.4797],[198,2.4376],[200,2.4619],[202,2.4779],[204,2.4888],[206,2.5762],[208,2.4686],[210,2.2487],[212,2.0706],[214,2],[218,1.7195],[220,1.6348],[222,1.6189],[226,1.6965],[228,1.7018],[230,1.6944],[232,1.6559],[236,1.5418],[238,1.5062],[240,1.5047],[242,1.499],[244,1.509],[248,1.5091],[250,1.5264],[252,1.5512],[254,1.553],[256,1.5332],[258,1.5253],[260,1.5216],[262,1.545],[264,1.6145],[266,1.668],[268,1.817],[270,1.92],[272,1.9261],[274,1.8606],[276,1.7666],[278,1.7429],[280,1.7948],[282,1.8743],[284,1.9905],[286,2.0848],[288,2.159],[290,2.2179],[292,2.2906],[294,2.3884],[296,2.4277],[298,2.5221],[300,2.589],[302,2.6412],[304,2.6687],[306,2.6556],[312,2.4044],[314,2.3031],[316,2.2187],[318,2.1855],[320,2.2188],[322,2.264],[324,2.4452],[326,2.6984],[328,3.0547],[330,3.4812],[332,4.0624],[334,5.3245],[336,8.8491],[338,12.6401],[340,16.214],[342,18.9337],[344,19.6717],[346,20.5725],[348,20.8965],[350,21.5544],[352,22.283],[354,22.5535],[358,23.2421],[360,23.6705],[362,23.7945],[364,31.5663],[366,45.3338],[368,60.5079]]
       }
       body_ctrl.Rotation.Y {
        type=MovieSceneScriptingFloatChannel
        default=3.8394
        thinned_from=140
        keys [[0,-8.8844],[92,-8.8844],[94,-8.9114],[96,-8.8382],[98,-8.6895],[100,-8.4813],[102,-8.3509],[104,-8.1118],[106,-7.4717],[108,-6.9164],[110,-6.0435],[112,-5.4876],[114,-4.8245],[116,-4.2763],[118,-3.7008],[120,-3.169],[122,-2.5709],[124,-2.2923],[126,-1.7284],[128,-1.1716],[130,-0.8908],[132,-0.3411],[134,0.0679],[136,0.6172],[138,1.0371],[140,1.2787],[142,1.3454],[144,1.1744],[146,0.8441],[148,0.6153],[150,0.3662],[152,0.0373],[154,-0.2408],[158,-0.5413],[160,-0.7372],[162,-0.989],[164,-1.1594],[166,-1.3043],[168,-1.396],[170,-1.422],[172,-1.3807],[174,-1.3287],[176,-1.301],[178,-1.254],[180,-1.1934],[182,-1.1361],[184,-1.0473],[186,-1.0145],[188,-0.9786],[190,-0.9695],[192,-0.9921],[194,-1.0759],[196,-1.2002],[198,-1.314],[200,-1.4352],[202,-1.5264],[204,-1.625],[206,-1.7883],[208,-1.8669],[210,-1.9291],[212,-1.983],[214,-2.018],[216,-2.1168],[220,-2.3286],[222,-2.4042],[224,-2.4341],[226,-2.4376],[228,-2.4509],[230,-2.4754],[232,-2.5172],[234,-2.5389],[236,-2.5813],[238,-2.6441],[240,-2.715],[242,-2.7569],[246,-2.9435],[248,-3],[252,-3.0464],[254,-3.0587],[256,-3.1062],[260,-3.1707],[264,-3.196],[268,-3.1958],[270,-3.1417],[272,-3.0688],[274,-3.0376],[276,-3.0686],[278,-3.1103],[280,-3.1228],[282,-3.119],[284,-3.1499],[288,-3.1727],[290,-3.151],[292,-3.0987],[294,-3.0593],[296,-3.0469],[298,-3.0501],[300,-3.0655],[302,-3.1155],[304,-3.1948],[306,-3.3098],[308,-3.5059],[310,-3.6628],[312,-3.8298],[316,-4.0294],[318,-4.0797],[320,-4.151],[322,-4.1813],[324,-4.2404],[326,-4.232],[330,-4.0498],[332,-3.965],[334,-3.9882],[336,-4.3964],[338,-5.0209],[340,-5.6343],[342,-6.1807],[344,-6.1724],[346,-5.9168],[348,-5.6198],[350,-5.0971],[352,-4.4942],[354,-4.045],[356,-3.6765],[358,-3.483],[360,-3.2495],[362,-2.8939],[364,0.3499],[366,2.2196],[368,3.8394]]
       }
       body_ctrl.Rotation.Z {
        type=MovieSceneScriptingFloatChannel
        default=10.253
        thinned_from=140
        keys [[0,10.7206],[92,10.7206],[94,10.772],[96,10.8531],[98,10.9783],[100,11.1372],[102,11.4056],[104,12.0735],[106,13.0679],[108,14.143],[110,15.5792],[112,16.8619],[114,18.0881],[116,19.2447],[118,20.1688],[120,20.8536],[122,21.3169],[124,21.6627],[126,22.1783],[128,22.8811],[130,23.3094],[132,23.8331],[134,23.9146],[136,23.6455],[138,23.5509],[142,23.4394],[144,23.6823],[148,25.0471],[152,26.4687],[154,26.9439],[158,27.3647],[160,27.674],[162,27.8995],[164,28.0667],[166,28.2912],[170,28.8958],[172,29.1365],[178,29.321],[190,29.4043],[194,29.3045],[198,28.9584],[200,28.6687],[202,28.4994],[212,27.9415],[216,27.8899],[220,27.9637],[224,27.904],[226,27.9163],[228,27.9979],[230,28.1979],[236,28.6019],[238,28.6748],[246,28.6798],[250,28.7672],[262,28.6777],[266,28.5472],[268,28.422],[286,28.1164],[290,28.0931],[292,28.0184],[294,27.8753],[296,27.8035],[302,27.3648],[304,27.2946],[306,27.3174],[318,27.9638],[320,28.0296],[322,28.0228],[328,27.7916],[332,27.4892],[334,27.2327],[336,26.2056],[338,25.8453],[340,25.3403],[342,25.6445],[346,26.5041],[348,27.2486],[350,27.8734],[352,28.0353],[354,27.9047],[356,27.9368],[358,27.347],[360,27.1591],[362,26.764],[364,24.9274],[366,18.7357],[368,10.253]]
       }
       body_ctrl.Scale.X {type=MovieSceneScriptingFloatChannel, default=1, const [1,140]}
       body_ctrl.Scale.Y {type=MovieSceneScriptingFloatChannel, default=1, const [1,140]}
       body_ctrl.Scale.Z {type=MovieSceneScriptingFloatChannel, default=1, const [1,140]}
       hips_ctrl.Location.X {type=MovieSceneScriptingFloatChannel, default=0, thinned_from=140, keys [[0,0],[368,0]]}
       hips_ctrl.Location.Y {type=MovieSceneScriptingFloatChannel, default=0, thinned_from=140, keys [[0,0],[368,0]]}
       hips_ctrl.Location.Z {type=MovieSceneScriptingFloatChannel, default=0, thinned_from=140, keys [[0,0],[368,0]]}
       hips_ctrl.Rotation.X {
        type=MovieSceneScriptingFloatChannel
        default=-5.577
        thinned_from=140
        keys [[0,-34.8803],[92,-34.8803],[94,-34.6599],[96,-34.2808],[98,-33.8253],[100,-33.2484],[102,-32.975],[104,-32.1802],[106,-30.2484],[108,-28.9049],[110,-26.0343],[112,-24.4421],[114,-22.5094],[116,-20.7856],[118,-18.8547],[122,-16.1793],[124,-15.7977],[126,-14.5191],[128,-13.7532],[130,-13.5841],[132,-12.8452],[134,-12.2221],[136,-11.6379],[138,-11.2324],[140,-10.9187],[146,-10.6249],[148,-10.4827],[152,-10.0248],[154,-9.8359],[158,-9.6643],[160,-9.3608],[162,-9.246],[164,-9.1646],[166,-9.1109],[170,-9.0755],[174,-9.0023],[180,-9.0993],[184,-9.2075],[188,-9.2527],[192,-9.3627],[196,-9.4056],[200,-9.4886],[202,-9.4766],[204,-9.4427],[206,-9.4874],[208,-9.3642],[210,-9.1178],[212,-8.933],[214,-8.8714],[216,-8.7638],[218,-8.6921],[220,-8.675],[222,-8.7229],[226,-8.8856],[228,-8.913],[230,-8.9593],[232,-8.9566],[234,-8.9759],[236,-8.9609],[238,-8.9224],[246,-8.9156],[252,-9.0469],[254,-9.0621],[256,-9.1238],[258,-9.1384],[260,-9.1331],[262,-9.1585],[266,-9.2672],[268,-9.3751],[270,-9.4239],[272,-9.3762],[276,-9.2082],[278,-9.2118],[284,-9.3324],[286,-9.4055],[288,-9.4518],[294,-9.4908],[300,-9.5023],[302,-9.5231],[308,-9.6768],[312,-9.716],[316,-9.6278],[318,-9.6146],[322,-9.6849],[324,-9.7866],[326,-9.924],[328,-10.1366],[330,-10.4352],[332,-10.9374],[334,-11.9536],[336,-14.6324],[338,-18.5565],[340,-22.6611],[342,-26.349],[344,-28.5099],[346,-30.7778],[348,-32.3339],[350,-33.4271],[352,-33.806],[354,-33.5696],[356,-33.2245],[358,-32.9727],[360,-32.4392],[362,-31.1307],[364,-26.4638],[366,-15.5518],[368,-5.577]]
       }
       hips_ctrl.Rotation.Y {
        type=MovieSceneScriptingFloatChannel
        default=-4.789
        thinned_from=140
        keys [[0,5.5057],[92,5.5057],[94,5.5912],[98,5.7233],[100,5.6995],[102,5.6464],[104,5.4407],[108,4.4254],[110,3.5867],[112,3.0608],[114,2.577],[116,2.3389],[118,2.1314],[120,1.8383],[122,1.2859],[124,1.0201],[126,0.6769],[128,0.4953],[130,0.3719],[132,0.0841],[134,-0.1424],[136,-0.4203],[138,-0.4828],[140,-0.4263],[142,-0.5457],[144,-0.5359],[146,-0.1587],[148,0.1187],[150,0.3583],[152,0.6729],[154,0.9591],[158,1.202],[160,1.4038],[162,1.6946],[164,1.9567],[166,2.1733],[168,2.286],[170,2.4039],[172,2.6242],[174,2.7194],[176,2.6852],[178,2.6134],[180,2.5817],[182,2.6162],[184,2.6304],[186,2.6254],[188,2.6146],[190,2.6327],[192,2.6734],[196,2.8143],[198,2.8703],[200,2.9338],[202,3.0178],[204,3.1466],[206,3.3612],[208,3.3944],[210,3.4197],[212,3.4314],[214,3.4335],[216,3.4451],[220,3.5778],[222,3.5443],[224,3.49],[226,3.4576],[228,3.4828],[230,3.5247],[234,3.4265],[236,3.4456],[238,3.5143],[240,3.5718],[242,3.6012],[244,3.6873],[246,3.7896],[248,3.8428],[250,3.8329],[252,3.8092],[254,3.8224],[256,3.8965],[258,3.9521],[262,3.9833],[264,4.0267],[266,4.0468],[268,4.1442],[270,4.1873],[274,4.0711],[278,4.085],[282,4.0724],[284,4.1437],[286,4.2019],[288,4.2833],[290,4.3347],[292,4.33],[296,4.3035],[298,4.3115],[300,4.3102],[302,4.3246],[304,4.3573],[306,4.4231],[308,4.5494],[312,4.7381],[316,4.8008],[318,4.8544],[320,4.9213],[322,4.9521],[324,5.007],[326,5.0223],[328,4.9158],[330,4.7601],[332,4.4623],[334,4.0959],[336,2.5954],[338,2.5432],[340,3.0154],[342,4.1658],[344,4.4053],[346,4.681],[348,4.8577],[350,5.2217],[352,5.4476],[354,5.7365],[356,6.2716],[358,6.4347],[360,6.5838],[362,6.1806],[364,0.9714],[366,-5.3726],[368,-4.789]]
       }
       hips_ctrl.Rotation.Z {
        type=MovieSceneScriptingFloatChannel
        default=0.8513
        thinned_from=140
        keys [[0,-2.4203],[92,-2.4203],[96,-2.5073],[98,-2.545],[100,-2.5295],[102,-2.4903],[104,-2.3271],[106,-1.9856],[108,-1.6898],[110,-1.2074],[112,-0.9215],[114,-0.6426],[116,-0.4762],[118,-0.3661],[120,-0.2768],[122,-0.1001],[124,-0.022],[126,0.0983],[128,0.2033],[130,0.2607],[132,0.3426],[134,0.3479],[136,0.3296],[138,0.2936],[140,0.2523],[142,0.2557],[144,0.2724],[146,0.2309],[148,0.2153],[150,0.2118],[152,0.1953],[154,0.164],[156,0.1494],[158,0.1383],[160,0.0746],[162,0.0139],[164,-0.0508],[166,-0.1056],[168,-0.1287],[170,-0.1588],[172,-0.2115],[174,-0.2368],[176,-0.2237],[178,-0.1999],[180,-0.191],[182,-0.1961],[184,-0.1965],[186,-0.1947],[188,-0.1967],[190,-0.2102],[192,-0.2389],[194,-0.281],[196,-0.3348],[198,-0.3837],[200,-0.4602],[202,-0.507],[204,-0.5387],[206,-0.5343],[212,-0.5354],[214,-0.5323],[216,-0.5263],[220,-0.5436],[222,-0.5294],[224,-0.5016],[226,-0.481],[228,-0.4736],[230,-0.4599],[232,-0.4287],[234,-0.3999],[236,-0.3892],[238,-0.3966],[240,-0.412],[242,-0.4222],[246,-0.4686],[248,-0.4775],[252,-0.4735],[254,-0.4798],[256,-0.496],[258,-0.5063],[260,-0.5038],[262,-0.5061],[264,-0.5196],[266,-0.5288],[268,-0.5614],[270,-0.5723],[272,-0.5612],[274,-0.5552],[276,-0.5515],[278,-0.5408],[280,-0.5222],[282,-0.512],[284,-0.537],[286,-0.5584],[288,-0.5851],[290,-0.6051],[292,-0.6117],[294,-0.6129],[296,-0.6112],[298,-0.6136],[300,-0.6117],[302,-0.615],[304,-0.6217],[306,-0.632],[308,-0.6558],[310,-0.6731],[312,-0.6883],[314,-0.6842],[316,-0.6829],[318,-0.6883],[320,-0.7043],[322,-0.7153],[324,-0.7432],[326,-0.7591],[328,-0.7515],[330,-0.7354],[332,-0.7243],[334,-0.7403],[336,-0.4889],[338,-0.5722],[340,-0.8829],[342,-1.2747],[344,-1.4133],[346,-1.5425],[348,-1.5445],[350,-1.6549],[352,-1.7747],[354,-1.9348],[356,-2.2246],[358,-2.3566],[360,-2.4641],[362,-2.3712],[364,-0.6276],[366,1.3753],[368,0.8513]]
       }
       hips_ctrl.Scale.X {type=MovieSceneScriptingFloatChannel, default=1, const [1,140]}
       hips_ctrl.Scale.Y {type=MovieSceneScriptingFloatChannel, default=1, const [1,140]}
       hips_ctrl.Scale.Z {type=MovieSceneScriptingFloatChannel, default=1, const [1,140]}
       hips_tan_ctrl.Location.X {
        type=MovieSceneScriptingFloatChannel
        default=-0.7493
        thinned_from=140
        keys [[0,0.8696],[92,0.8696],[94,0.8821],[98,0.8997],[100,0.8948],[102,0.8863],[104,0.8558],[108,0.7015],[110,0.5735],[112,0.4926],[114,0.4186],[116,0.3822],[118,0.3478],[120,0.2991],[122,0.2116],[124,0.1695],[126,0.1158],[128,0.0882],[130,0.0693],[132,0.024],[134,-0.0134],[136,-0.06],[138,-0.0723],[140,-0.0651],[142,-0.0843],[144,-0.0822],[146,-0.0229],[148,0.0207],[150,0.0585],[152,0.1081],[154,0.1526],[156,0.1712],[158,0.1902],[160,0.2204],[162,0.2653],[164,0.3056],[166,0.3389],[170,0.3742],[172,0.4078],[174,0.4222],[176,0.417],[178,0.406],[180,0.401],[182,0.4062],[184,0.4082],[186,0.4074],[188,0.4055],[190,0.4081],[192,0.4142],[196,0.4355],[198,0.4438],[200,0.4532],[202,0.4659],[204,0.4862],[206,0.5205],[208,0.5256],[210,0.5297],[212,0.5316],[214,0.5319],[216,0.5337],[220,0.5544],[222,0.5494],[224,0.5412],[226,0.5364],[228,0.5406],[230,0.5473],[234,0.5326],[236,0.5357],[238,0.5464],[240,0.5554],[242,0.56],[244,0.573],[246,0.5891],[248,0.5974],[250,0.5954],[252,0.5917],[254,0.5938],[256,0.6055],[258,0.6142],[262,0.6191],[264,0.6259],[266,0.6291],[268,0.6437],[270,0.6504],[272,0.6405],[274,0.6317],[278,0.6332],[282,0.6324],[284,0.6436],[286,0.6525],[288,0.6653],[290,0.6728],[292,0.6721],[296,0.6676],[300,0.6686],[302,0.6709],[304,0.6766],[306,0.6869],[308,0.7074],[312,0.7375],[314,0.7416],[316,0.7478],[318,0.7561],[320,0.7667],[322,0.7716],[324,0.7804],[326,0.7829],[328,0.7653],[330,0.741],[332,0.6952],[334,0.6378],[336,0.4058],[338,0.3975],[340,0.4693],[342,0.6465],[344,0.6801],[346,0.7196],[348,0.7449],[350,0.7994],[352,0.8297],[354,0.871],[356,0.9548],[360,1.02],[362,0.9639],[364,0.1436],[366,-0.8426],[368,-0.7493]]
       }
       hips_tan_ctrl.Location.Y {
        type=MovieSceneScriptingFloatChannel
        default=1.0811
        thinned_from=140
        keys [[0,6.719],[92,6.719],[94,6.6916],[98,6.5897],[100,6.5198],[102,6.4749],[104,6.3451],[106,6.0544],[108,5.8183],[110,5.3658],[112,5.0651],[114,4.7067],[116,4.3723],[118,4.0061],[122,3.4597],[124,3.3461],[126,3.1056],[128,2.9393],[130,2.8639],[132,2.7031],[134,2.5601],[136,2.4224],[138,2.3129],[140,2.2184],[144,2.0989],[146,2.0426],[148,1.9818],[152,1.8419],[154,1.7821],[158,1.7046],[160,1.6367],[162,1.6024],[164,1.5773],[166,1.5585],[170,1.5323],[174,1.5013],[176,1.495],[180,1.49],[184,1.4915],[188,1.4858],[192,1.4927],[196,1.4928],[200,1.5033],[202,1.5013],[204,1.4959],[206,1.5027],[208,1.4836],[210,1.4456],[212,1.417],[214,1.4074],[216,1.3907],[218,1.3803],[220,1.3771],[222,1.3845],[226,1.4095],[228,1.4136],[230,1.4206],[232,1.42],[234,1.4229],[236,1.4205],[238,1.4144],[246,1.4128],[252,1.4324],[254,1.4347],[256,1.4441],[258,1.4462],[260,1.4454],[262,1.4493],[266,1.466],[268,1.4823],[270,1.4898],[272,1.4822],[276,1.4564],[278,1.4564],[284,1.4753],[286,1.4862],[288,1.4932],[294,1.4992],[300,1.501],[302,1.5042],[308,1.5266],[312,1.5317],[314,1.5262],[316,1.5175],[318,1.5153],[322,1.5257],[324,1.5412],[326,1.5623],[328,1.5956],[330,1.642],[332,1.7196],[334,1.8773],[336,2.2899],[340,3.459],[342,3.9548],[346,4.4985],[348,4.6688],[350,4.7803],[352,4.8085],[354,4.7701],[356,4.7468],[358,4.7698],[360,4.768],[362,4.6583],[364,4.1211],[366,2.5829],[368,1.0811]]
       }
       hips_tan_ctrl.Location.Z {
        type=MovieSceneScriptingFloatChannel
        default=-0.1086
        thinned_from=140
        keys [[0,-2.222],[92,-2.222],[94,-2.1977],[96,-2.1555],[98,-2.105],[100,-2.0407],[102,-2.0073],[104,-1.9115],[106,-1.6934],[108,-1.5417],[110,-1.2528],[112,-1.0976],[114,-0.9252],[118,-0.6401],[120,-0.546],[122,-0.4587],[124,-0.4323],[126,-0.3617],[128,-0.3202],[130,-0.3091],[132,-0.273],[134,-0.2442],[136,-0.2185],[138,-0.2007],[140,-0.1867],[144,-0.174],[146,-0.1672],[148,-0.1596],[150,-0.1499],[152,-0.1409],[154,-0.1336],[158,-0.1251],[160,-0.1161],[162,-0.113],[164,-0.1112],[166,-0.1101],[170,-0.1086],[172,-0.1081],[174,-0.1071],[178,-0.1057],[180,-0.1055],[184,-0.1065],[186,-0.106],[188,-0.1059],[192,-0.1074],[194,-0.1079],[196,-0.1086],[200,-0.111],[204,-0.1117],[206,-0.1146],[208,-0.1122],[210,-0.1072],[212,-0.1034],[214,-0.1022],[216,-0.1001],[218,-0.0993],[220,-0.0996],[222,-0.1002],[226,-0.1027],[228,-0.1035],[230,-0.1049],[232,-0.1043],[238,-0.1039],[240,-0.1045],[242,-0.1047],[244,-0.1054],[246,-0.1065],[248,-0.1077],[252,-0.1093],[254,-0.1098],[256,-0.1118],[258,-0.1127],[260,-0.1127],[262,-0.1134],[264,-0.1152],[266,-0.1165],[268,-0.1198],[270,-0.1213],[272,-0.1195],[274,-0.117],[276,-0.1154],[278,-0.1154],[280,-0.1163],[282,-0.117],[284,-0.1188],[288,-0.1229],[292,-0.1241],[298,-0.1239],[300,-0.1243],[302,-0.1249],[304,-0.1264],[306,-0.1285],[308,-0.131],[310,-0.1328],[312,-0.1342],[314,-0.1337],[316,-0.1329],[318,-0.1333],[320,-0.1349],[322,-0.1361],[324,-0.1392],[328,-0.1462],[330,-0.1514],[332,-0.1606],[334,-0.1842],[336,-0.2679],[338,-0.4531],[340,-0.7021],[342,-0.9765],[344,-1.1677],[346,-1.3809],[348,-1.5325],[350,-1.6567],[352,-1.7047],[354,-1.6972],[356,-1.6545],[358,-1.5662],[360,-1.4538],[362,-1.2839],[364,-0.9127],[366,-0.3889],[368,-0.1086]]
       }
       hips_tan_ctrl.Rotation.X {type=MovieSceneScriptingFloatChannel, default=0, const [0,140]}
       hips_tan_ctrl.Rotation.Y {type=MovieSceneScriptingFloatChannel, default=0, const [0,140]}
       hips_tan_ctrl.Rotation.Z {type=MovieSceneScriptingFloatChannel, default=0, const [0,140]}
       hips_tan_ctrl.Scale.X {type=MovieSceneScriptingFloatChannel, default=1, const [1,140]}
       hips_tan_ctrl.Scale.Y {type=MovieSceneScriptingFloatChannel, default=1, const [1,140]}
       hips_tan_ctrl.Scale.Z {type=MovieSceneScriptingFloatChannel, default=1, const [1,140]}
       spine_01_ctrl.Location.X {type=MovieSceneScriptingFloatChannel, default=0, thinned_from=140, keys [[0,0],[358,0],[360,-0.0001],[364,0.0002],[368,0]]}
       spine_01_ctrl.Location.Y {
        type=MovieSceneScriptingFloatChannel
        default=0
        thinned_from=140
        keys [[0,0],[344,0],[346,-0.0027],[348,0],[350,-0.0026],[352,0],[354,0],[356,-0.0027],[358,0],[360,-0.0027],[362,-0.003],[364,-0.0045],[366,0],[368,0]]
       }
       spine_01_ctrl.Location.Z {
        type=MovieSceneScriptingFloatChannel
        default=0
        thinned_from=134
        keys [[0,0],[102,-0.0001],[110,-0.0004],[112,-0.0003],[114,-0.0004],[116,-0.0002],[118,-0.0005],[120,-0.0001],[124,0],[126,-0.0003],[128,0],[330,0],[332,-0.0001],[334,0],[340,-0.0003],[344,0],[366,-0.0001],[368,0]]
       }
       spine_01_ctrl.Rotation.X {
        type=MovieSceneScriptingFloatChannel
        default=-6.7084
        thinned_from=140
        keys [[0,-28.5971],[100,-28.5391],[102,-28.3445],[104,-27.8004],[106,-26.963],[108,-25.9078],[110,-24.7032],[112,-23.3345],[116,-20.4146],[118,-19],[120,-17.668],[122,-16.4817],[124,-15.5581],[126,-14.7653],[128,-13.9187],[130,-13.1052],[138,-9.9738],[140,-9.2101],[144,-7.7302],[146,-7.0336],[148,-6.391],[150,-5.7998],[152,-5.2578],[154,-4.7574],[156,-4.3159],[158,-3.9271],[160,-3.5758],[162,-3.3098],[164,-3.1051],[170,-2.5919],[172,-2.4255],[176,-2.1162],[178,-1.9694],[180,-1.8323],[182,-1.7027],[184,-1.5812],[186,-1.4701],[188,-1.3701],[190,-1.2773],[192,-1.1973],[194,-1.1369],[196,-1.0892],[198,-1.0532],[200,-1.0478],[202,-1.0494],[204,-1.0468],[206,-1.0571],[216,-1.0668],[218,-1.0721],[220,-1.0731],[222,-1.0715],[228,-1.058],[230,-1.056],[232,-1.0485],[234,-1.0441],[236,-1.0425],[240,-1.0451],[244,-1.0405],[246,-1.0403],[250,-1.0315],[254,-1.0283],[256,-1.0187],[262,-1.0226],[270,-1.0131],[272,-1.008],[274,-1.0106],[278,-1.0097],[282,-1.006],[286,-0.9969],[288,-0.995],[296,-1.0042],[298,-1.0097],[300,-1.01],[304,-1.0046],[306,-0.9964],[308,-0.9793],[310,-0.9701],[314,-0.9555],[318,-0.9467],[320,-0.9441],[322,-0.9435],[326,-0.9474],[328,-0.9476],[330,-0.9459],[332,-0.941],[334,-0.958],[336,-0.9557],[338,-0.7414],[340,-0.4396],[342,-0.0212],[344,0.3585],[346,1.096],[348,1.1257],[350,1.6974],[352,1.5795],[354,1.6208],[356,1.6826],[358,0.7725],[360,0.3556],[362,-0.4143],[364,-1.4261],[366,-6.0929],[368,-6.7084]]
       }
       spine_01_ctrl.Rotation.Y {
        type=MovieSceneScriptingFloatChannel
        default=-0.0042
        thinned_from=140
        keys [[0,0.2734],[92,0.2734],[94,0.2657],[96,0.2471],[98,0.2223],[100,0.1955],[102,0.1834],[104,0.1829],[106,0.1744],[108,0.1626],[110,0.1482],[112,0.1516],[114,0.1654],[116,0.1664],[118,0.1475],[120,0.1358],[122,0.158],[124,0.1664],[126,0.17],[128,0.171],[130,0.177],[132,0.1812],[134,0.1751],[136,0.1474],[138,0.106],[140,0.0752],[142,0.0757],[144,0.0868],[146,0.075],[148,0.0607],[150,0.0542],[152,0.0485],[154,0.0345],[156,0.0262],[158,0.0218],[160,-0.0152],[162,-0.0342],[164,-0.0564],[166,-0.0782],[168,-0.0886],[170,-0.1026],[172,-0.1308],[174,-0.1478],[176,-0.1427],[178,-0.1336],[180,-0.1362],[182,-0.144],[184,-0.1495],[186,-0.1514],[188,-0.1558],[190,-0.1636],[192,-0.1747],[194,-0.1843],[196,-0.1993],[198,-0.213],[200,-0.2308],[202,-0.2434],[204,-0.2542],[206,-0.2275],[208,-0.2244],[214,-0.218],[216,-0.2153],[218,-0.2194],[220,-0.2222],[222,-0.2181],[224,-0.209],[226,-0.2004],[228,-0.1951],[230,-0.1904],[232,-0.1789],[234,-0.1687],[236,-0.1668],[238,-0.17],[240,-0.1757],[242,-0.1803],[244,-0.1859],[246,-0.1903],[248,-0.1927],[250,-0.193],[252,-0.1907],[254,-0.1935],[256,-0.1998],[258,-0.2027],[260,-0.2008],[262,-0.2014],[264,-0.2069],[266,-0.2104],[268,-0.2225],[270,-0.2312],[272,-0.2353],[276,-0.2382],[278,-0.2338],[280,-0.2267],[282,-0.2237],[284,-0.2275],[286,-0.2328],[290,-0.2493],[292,-0.2557],[294,-0.2587],[298,-0.2572],[300,-0.2546],[302,-0.2498],[304,-0.2422],[306,-0.2342],[308,-0.2239],[312,-0.2078],[316,-0.1932],[320,-0.1914],[322,-0.1916],[324,-0.1956],[326,-0.202],[328,-0.2029],[330,-0.1947],[332,-0.1904],[334,-0.1928],[336,-0.0778],[338,-0.071],[340,-0.1049],[342,-0.1224],[344,-0.1317],[346,-0.0613],[348,-0.0857],[350,-0.0259],[352,-0.0831],[354,-0.1178],[356,-0.0689],[358,-0.209],[360,-0.105],[362,-0.1369],[364,-0.1603],[366,-0.0598],[368,-0.0042]]
       }
       spine_01_ctrl.Rotation.Z {
        type=MovieSceneScriptingFloatChannel
        default=0.9333
        thinned_from=140
        keys [[0,-1.0895],[92,-1.0895],[94,-1.091],[96,-1.0824],[98,-1.0793],[100,-1.0587],[102,-1.0409],[104,-1.0075],[106,-0.9221],[108,-0.8606],[110,-0.7634],[112,-0.6914],[114,-0.6334],[116,-0.6161],[118,-0.5403],[120,-0.4028],[122,-0.2602],[124,-0.2044],[126,-0.1606],[128,-0.1967],[130,-0.2125],[132,-0.1882],[134,-0.0864],[136,0.0671],[138,0.1489],[140,0.1934],[142,0.2458],[144,0.2286],[146,0.1478],[148,0.0577],[150,-0.036],[152,-0.1362],[154,-0.2017],[156,-0.2269],[158,-0.2585],[160,-0.2306],[162,-0.2411],[164,-0.2338],[166,-0.2289],[168,-0.2374],[170,-0.2339],[174,-0.2312],[176,-0.2405],[178,-0.2561],[180,-0.2659],[184,-0.2939],[186,-0.2987],[188,-0.2966],[190,-0.2864],[192,-0.2588],[194,-0.2164],[198,-0.0974],[200,0.0052],[202,0.0497],[204,0.0516],[206,-0.0426],[208,-0.0375],[210,-0.0281],[212,-0.023],[214,-0.0243],[216,-0.0327],[218,-0.0403],[220,-0.0519],[222,-0.067],[224,-0.1007],[226,-0.1309],[228,-0.1584],[230,-0.2048],[232,-0.2373],[234,-0.2636],[236,-0.287],[238,-0.3003],[240,-0.2986],[244,-0.287],[246,-0.2887],[248,-0.2968],[250,-0.3002],[252,-0.2982],[254,-0.2948],[256,-0.3001],[258,-0.3061],[260,-0.3159],[262,-0.3203],[264,-0.3206],[268,-0.3129],[270,-0.3185],[272,-0.3085],[274,-0.2889],[276,-0.2925],[278,-0.3134],[280,-0.3472],[282,-0.3679],[284,-0.3591],[286,-0.3531],[288,-0.346],[290,-0.3352],[292,-0.3256],[294,-0.3197],[296,-0.3183],[298,-0.3196],[300,-0.3239],[302,-0.326],[304,-0.3307],[306,-0.3419],[308,-0.3499],[310,-0.3566],[312,-0.3624],[314,-0.373],[316,-0.3822],[318,-0.3935],[320,-0.3981],[322,-0.3963],[324,-0.388],[326,-0.391],[328,-0.3906],[330,-0.3868],[332,-0.3462],[334,-0.3001],[336,-0.3094],[338,-0.4109],[340,-0.4737],[342,-0.7537],[344,-0.8313],[346,-0.5259],[348,-1.0354],[350,-0.5718],[352,-1.1528],[354,-1.1477],[356,-0.5548],[358,-1.1036],[360,-0.5398],[362,-0.537],[364,-0.5292],[366,0.5359],[368,0.9333]]
       }
       spine_01_ctrl.Scale.X {type=MovieSceneScriptingFloatChannel, default=1, const [1,140]}
       spine_01_ctrl.Scale.Y {type=MovieSceneScriptingFloatChannel, default=1, const [1,140]}
       spine_01_ctrl.Scale.Z {type=MovieSceneScriptingFloatChannel, default=1, const [1,140]}
       spine_02_ctrl.Location.X {
        type=MovieSceneScriptingFloatChannel
        default=0
        thinned_from=140
        keys [[0,0],[344,0],[346,0.0033],[348,0],[350,-0.0005],[352,0],[354,0],[356,0.0055],[358,0],[360,0.0101],[362,0.0134],[364,0.0195],[366,0],[368,0]]
       }
       spine_02_ctrl.Location.Y {
        type=MovieSceneScriptingFloatChannel
        default=0
        thinned_from=140
        keys [[0,0],[344,0],[346,-0.0434],[348,0],[350,-0.0397],[352,0],[354,0],[356,-0.0422],[358,0],[360,-0.0522],[362,-0.0789],[364,-0.2142],[366,0],[368,0]]
       }
       spine_02_ctrl.Location.Z {
        type=MovieSceneScriptingFloatChannel
        default=0
        thinned_from=140
        keys [[0,0],[112,-0.0002],[114,-0.0004],[116,-0.0001],[118,-0.0004],[120,-0.0001],[124,0],[126,-0.0002],[128,0],[334,0],[340,-0.0002],[344,0],[346,0.0445],[348,0],[350,0.0449],[352,0],[354,0],[356,0.0447],[358,0],[360,0.0438],[362,0.041],[364,0.0402],[366,-0.0001],[368,0]]
       }
       spine_02_ctrl.Rotation.X {
        type=MovieSceneScriptingFloatChannel
        default=-7.1335
        thinned_from=140
        keys [[0,-11.8393],[92,-11.8393],[100,-11.793],[102,-11.7227],[104,-11.5221],[106,-11.2109],[108,-10.831],[110,-10.4135],[112,-9.9105],[116,-8.8114],[120,-7.7647],[122,-7.2932],[124,-6.9473],[126,-6.6541],[128,-6.311],[130,-6.0021],[134,-5.443],[140,-4.5587],[142,-4.2758],[144,-3.9766],[146,-3.7001],[148,-3.4555],[150,-3.2377],[152,-3.042],[154,-2.8582],[156,-2.7006],[160,-2.4205],[162,-2.3255],[164,-2.2488],[168,-2.1195],[170,-2.0495],[172,-1.9847],[174,-1.9303],[176,-1.8699],[178,-1.8148],[182,-1.7172],[184,-1.6719],[186,-1.6303],[190,-1.5544],[192,-1.5204],[196,-1.4761],[198,-1.4565],[202,-1.4575],[204,-1.4549],[206,-1.466],[216,-1.4759],[218,-1.4813],[222,-1.4812],[228,-1.4689],[230,-1.4674],[232,-1.4604],[234,-1.4563],[240,-1.4583],[244,-1.4541],[246,-1.4543],[250,-1.4461],[254,-1.4433],[256,-1.4342],[262,-1.4392],[270,-1.4311],[272,-1.4261],[278,-1.4288],[282,-1.4263],[286,-1.418],[288,-1.4165],[296,-1.4269],[298,-1.4328],[302,-1.4323],[304,-1.4291],[306,-1.4215],[308,-1.4051],[314,-1.3831],[320,-1.3735],[324,-1.3752],[328,-1.3787],[330,-1.3771],[332,-1.3718],[334,-1.3883],[336,-1.348],[338,-1.0328],[340,-0.5825],[342,0.0131],[344,0.5896],[346,1.2213],[348,1.7336],[350,2.198],[352,2.4352],[354,2.5121],[356,2.1508],[358,1.1867],[360,-0.0315],[362,-1.4659],[364,-4.2369],[366,-6.5158],[368,-7.1335]]
       }
       spine_02_ctrl.Rotation.Y {
        type=MovieSceneScriptingFloatChannel
        default=0.0703
        thinned_from=140
        keys [[0,0.0639],[92,0.0639],[94,0.0573],[96,0.0429],[98,0.0219],[100,0.0029],[102,-0.0034],[104,0.0061],[106,0.0199],[108,0.0236],[110,0.0329],[112,0.0516],[114,0.0784],[116,0.0863],[118,0.0822],[120,0.0899],[122,0.1292],[124,0.1437],[126,0.1522],[128,0.1499],[130,0.1544],[132,0.161],[134,0.164],[136,0.1494],[138,0.1149],[140,0.0879],[142,0.0924],[144,0.102],[146,0.0844],[148,0.0633],[150,0.0497],[152,0.0368],[154,0.0184],[156,0.0084],[158,0.0019],[160,-0.032],[162,-0.051],[166,-0.0926],[168,-0.1034],[170,-0.1168],[172,-0.1443],[174,-0.1608],[176,-0.1566],[178,-0.1489],[180,-0.1523],[182,-0.1611],[184,-0.1678],[186,-0.17],[188,-0.1743],[190,-0.1814],[192,-0.1902],[194,-0.1963],[196,-0.2064],[198,-0.2154],[200,-0.2248],[202,-0.2337],[204,-0.244],[206,-0.2242],[208,-0.2205],[212,-0.2152],[216,-0.2111],[218,-0.2156],[220,-0.2191],[222,-0.2163],[226,-0.2037],[228,-0.2005],[230,-0.1993],[232,-0.1905],[234,-0.1824],[236,-0.1822],[238,-0.1863],[240,-0.1917],[242,-0.1958],[244,-0.2007],[246,-0.2049],[248,-0.2079],[250,-0.2085],[252,-0.2062],[254,-0.2087],[256,-0.2153],[258,-0.2185],[260,-0.2174],[262,-0.2182],[264,-0.2238],[266,-0.227],[268,-0.2386],[270,-0.2477],[272,-0.2511],[274,-0.2512],[276,-0.2527],[278,-0.25],[280,-0.2456],[282,-0.2442],[284,-0.2472],[286,-0.2519],[290,-0.2668],[292,-0.2724],[294,-0.275],[298,-0.2735],[300,-0.2712],[302,-0.2666],[306,-0.2522],[308,-0.2424],[312,-0.227],[316,-0.2138],[318,-0.2138],[320,-0.213],[322,-0.213],[324,-0.2164],[326,-0.2232],[328,-0.2245],[330,-0.2168],[332,-0.2101],[334,-0.21],[336,-0.0998],[338,-0.1077],[340,-0.1544],[342,-0.2131],[344,-0.2388],[346,-0.2497],[348,-0.2372],[350,-0.2179],[352,-0.2576],[354,-0.2894],[356,-0.3352],[358,-0.3675],[360,-0.4143],[362,-0.4648],[364,-0.4347],[366,-0.0142],[368,0.0703]]
       }
       spine_02_ctrl.Rotation.Z {
        type=MovieSceneScriptingFloatChannel
        default=0.8379
        thinned_from=140
        keys [[0,-1.0526],[92,-1.0526],[94,-1.0515],[96,-1.0385],[98,-1.0297],[100,-1.0049],[102,-0.9858],[104,-0.956],[106,-0.8778],[108,-0.822],[110,-0.7353],[112,-0.672],[114,-0.6235],[116,-0.6099],[118,-0.5349],[120,-0.4005],[122,-0.2681],[124,-0.2168],[126,-0.1776],[128,-0.216],[130,-0.2339],[132,-0.2134],[134,-0.1138],[136,0.0386],[138,0.1229],[140,0.1706],[142,0.2213],[144,0.2034],[146,0.1284],[148,0.0431],[150,-0.047],[152,-0.1426],[154,-0.2032],[156,-0.2262],[158,-0.2558],[160,-0.2226],[162,-0.2279],[164,-0.2156],[166,-0.2064],[168,-0.2127],[172,-0.2006],[174,-0.1968],[176,-0.2069],[178,-0.224],[180,-0.234],[182,-0.2469],[184,-0.2603],[186,-0.265],[188,-0.2627],[190,-0.2517],[192,-0.2228],[194,-0.1789],[196,-0.1173],[198,-0.0562],[200,0.0484],[202,0.095],[204,0.0993],[206,0.0062],[208,0.0112],[210,0.0207],[212,0.0257],[214,0.0243],[216,0.0158],[218,0.0094],[220,-0.0012],[222,-0.017],[224,-0.0519],[226,-0.083],[228,-0.1106],[230,-0.1566],[232,-0.1907],[234,-0.2183],[236,-0.2416],[238,-0.2538],[240,-0.2509],[242,-0.2443],[244,-0.2371],[246,-0.2371],[248,-0.2444],[250,-0.2479],[252,-0.2464],[254,-0.2426],[256,-0.2466],[258,-0.2515],[260,-0.2614],[262,-0.2654],[264,-0.2647],[266,-0.2607],[268,-0.2544],[270,-0.2588],[272,-0.2494],[274,-0.2305],[276,-0.2339],[278,-0.2551],[280,-0.2894],[282,-0.3103],[284,-0.3004],[286,-0.2933],[288,-0.2845],[290,-0.2725],[292,-0.2624],[294,-0.2564],[296,-0.2552],[298,-0.2563],[300,-0.2609],[302,-0.2631],[304,-0.2679],[306,-0.2788],[308,-0.2861],[310,-0.2922],[312,-0.2975],[314,-0.3084],[316,-0.3178],[318,-0.3286],[320,-0.3324],[322,-0.3301],[324,-0.3208],[326,-0.3229],[328,-0.3235],[330,-0.3219],[332,-0.2843],[334,-0.2407],[336,-0.2688],[338,-0.3717],[340,-0.423],[342,-0.6867],[344,-0.7589],[346,-1.121],[348,-0.9631],[350,-1.4612],[352,-1.0741],[354,-1.0606],[356,-1.4417],[358,-0.9945],[360,-1.3122],[362,-1.115],[364,-0.2561],[366,0.4381],[368,0.8379]]
       }
       spine_02_ctrl.Scale.X {type=MovieSceneScriptingFloatChannel, default=1, const [1,140]}
       spine_02_ctrl.Scale.Y {type=MovieSceneScriptingFloatChannel, default=1, const [1,140]}
       spine_02_ctrl.Scale.Z {type=MovieSceneScriptingFloatChannel, default=1, const [1,140]}
       spine_03_ctrl.Location.X {
        type=MovieSceneScriptingFloatChannel
        default=0
        thinned_from=140
        keys [[0,0],[334,0],[336,-0.0002],[338,-0.0007],[340,-0.0016],[344,-0.0038],[346,0.0034],[348,-0.006],[350,0.0006],[352,-0.0074],[354,-0.0077],[356,0.0054],[358,-0.005],[360,0.0126],[362,0.0162],[364,0.0176],[366,0],[368,0]]
       }
       spine_03_ctrl.Location.Y {
        type=MovieSceneScriptingFloatChannel
        default=0
        thinned_from=140
        keys [[0,0],[334,0],[336,0.0013],[338,0.005],[340,0.0103],[342,0.0168],[344,0.0238],[346,-0.0114],[348,0.0374],[350,0.0031],[352,0.0463],[354,0.0477],[356,0.0007],[358,0.0309],[360,-0.0335],[362,-0.067],[364,-0.1843],[366,0],[368,0]]
       }
       spine_03_ctrl.Location.Z {
        type=MovieSceneScriptingFloatChannel
        default=0
        thinned_from=140
        keys [[0,0],[98,0],[110,-0.0004],[112,-0.0003],[114,-0.0005],[116,-0.0002],[118,-0.0005],[120,-0.0001],[124,0],[126,-0.0003],[128,0],[330,0],[332,-0.0001],[334,0],[336,-0.0308],[338,-0.1142],[340,-0.2371],[342,-0.3861],[344,-0.5482],[346,-0.8079],[348,-0.8595],[350,-1.0794],[352,-1.0656],[354,-1.0963],[356,-1.0797],[358,-0.7104],[360,-0.4843],[362,-0.2154],[364,-0.1401],[366,-0.0002],[368,0]]
       }
       spine_03_ctrl.Rotation.X {
        type=MovieSceneScriptingFloatChannel
        default=-7.8027
        thinned_from=140
        keys [[0,-11.0439],[92,-11.0439],[100,-11.0018],[102,-10.9734],[104,-10.8885],[110,-10.4479],[112,-10.2367],[122,-9.05],[124,-8.8994],[126,-8.7751],[128,-8.6021],[130,-8.4639],[134,-8.2442],[142,-7.7231],[144,-7.5728],[146,-7.438],[148,-7.3279],[154,-7.0811],[160,-6.8954],[172,-6.7018],[192,-6.4961],[198,-6.4588],[204,-6.4579],[222,-6.4845],[254,-6.446],[304,-6.4242],[322,-6.366],[332,-6.3613],[334,-6.3767],[336,-6.335],[338,-6.0198],[340,-5.5698],[342,-4.9777],[344,-4.4018],[346,-4.1682],[348,-3.2593],[350,-3.1874],[352,-2.5591],[354,-2.4826],[356,-3.2488],[358,-3.8086],[360,-5.3372],[362,-6.4569],[364,-8.0862],[366,-7.7753],[368,-7.8027]]
       }
       spine_03_ctrl.Rotation.Y {
        type=MovieSceneScriptingFloatChannel
        default=0.1338
        thinned_from=140
        keys [[0,-0.1288],[92,-0.1288],[94,-0.134],[96,-0.1437],[98,-0.1603],[100,-0.1714],[102,-0.1721],[104,-0.1538],[106,-0.12],[108,-0.1028],[110,-0.0726],[112,-0.0405],[114,-0.0024],[116,0.0117],[118,0.0216],[120,0.0473],[122,0.1019],[124,0.1219],[126,0.1346],[128,0.1293],[130,0.1322],[132,0.1408],[134,0.1522],[136,0.1495],[138,0.1214],[140,0.0982],[144,0.1144],[146,0.0918],[148,0.0648],[150,0.0449],[152,0.0257],[154,0.0036],[156,-0.0077],[158,-0.0161],[160,-0.0468],[162,-0.0656],[164,-0.0849],[166,-0.1045],[168,-0.1155],[170,-0.1282],[172,-0.1548],[174,-0.1708],[176,-0.1674],[178,-0.1611],[180,-0.1653],[182,-0.175],[184,-0.1826],[186,-0.1852],[188,-0.1894],[192,-0.2024],[194,-0.2052],[196,-0.2108],[198,-0.2153],[200,-0.2172],[202,-0.2225],[204,-0.2323],[206,-0.2187],[208,-0.2143],[210,-0.211],[212,-0.2082],[216,-0.2048],[218,-0.2096],[220,-0.2137],[222,-0.212],[226,-0.2042],[228,-0.2029],[230,-0.2048],[232,-0.1984],[234,-0.1924],[236,-0.1938],[240,-0.2036],[242,-0.2072],[246,-0.2155],[248,-0.2188],[250,-0.2198],[252,-0.2175],[254,-0.2197],[256,-0.2265],[258,-0.2299],[260,-0.2294],[262,-0.2306],[264,-0.236],[266,-0.239],[268,-0.25],[270,-0.2594],[272,-0.2622],[274,-0.261],[276,-0.2627],[278,-0.2614],[280,-0.2595],[282,-0.2597],[284,-0.2619],[286,-0.2661],[288,-0.273],[290,-0.2792],[292,-0.2842],[294,-0.2864],[298,-0.2848],[300,-0.2828],[302,-0.2784],[306,-0.265],[308,-0.2556],[312,-0.2407],[316,-0.2288],[318,-0.2295],[320,-0.2288],[322,-0.2287],[324,-0.2314],[326,-0.2384],[328,-0.2403],[330,-0.233],[332,-0.2244],[334,-0.2219],[336,-0.1191],[338,-0.14],[340,-0.1962],[342,-0.2921],[344,-0.3324],[346,-0.4502],[348,-0.3705],[350,-0.4806],[352,-0.4136],[354,-0.4414],[356,-0.6044],[358,-0.5047],[360,-0.6685],[362,-0.6924],[364,-0.5543],[366,0.0201],[368,0.1338]]
       }
       spine_03_ctrl.Rotation.Z {
        type=MovieSceneScriptingFloatChannel
        default=0.7395
        thinned_from=140
        keys [[0,-0.9849],[92,-0.9849],[94,-0.9815],[96,-0.9652],[98,-0.952],[100,-0.9246],[102,-0.9054],[104,-0.8812],[106,-0.8144],[108,-0.7671],[110,-0.6948],[112,-0.6423],[114,-0.6053],[116,-0.5964],[118,-0.524],[120,-0.3943],[122,-0.2733],[124,-0.2266],[126,-0.1924],[128,-0.233],[130,-0.2529],[132,-0.2365],[134,-0.1394],[136,0.0111],[138,0.0974],[140,0.1478],[142,0.1967],[144,0.1784],[146,0.1093],[148,0.0289],[150,-0.0573],[152,-0.1481],[154,-0.2038],[156,-0.2245],[158,-0.2522],[160,-0.2141],[162,-0.2144],[164,-0.1973],[166,-0.1839],[168,-0.188],[170,-0.1796],[172,-0.1692],[174,-0.163],[176,-0.1739],[178,-0.1925],[180,-0.2026],[182,-0.2145],[184,-0.2272],[186,-0.2318],[188,-0.2294],[190,-0.2177],[192,-0.1877],[194,-0.1423],[196,-0.079],[198,-0.0166],[200,0.0898],[202,0.1381],[204,0.1446],[206,0.0531],[208,0.0582],[210,0.0677],[212,0.0727],[214,0.0711],[216,0.0626],[218,0.0573],[220,0.0478],[222,0.0313],[224,-0.0046],[226,-0.0366],[228,-0.064],[230,-0.1096],[232,-0.1449],[234,-0.1737],[236,-0.1969],[238,-0.208],[240,-0.2039],[242,-0.1967],[244,-0.1881],[246,-0.1865],[248,-0.193],[250,-0.1965],[252,-0.1953],[254,-0.1913],[256,-0.194],[258,-0.198],[260,-0.2077],[262,-0.2114],[264,-0.2098],[266,-0.2053],[268,-0.1969],[270,-0.2002],[272,-0.1916],[274,-0.1733],[276,-0.1766],[278,-0.1978],[280,-0.2326],[282,-0.2536],[284,-0.2426],[286,-0.2343],[288,-0.2241],[290,-0.2109],[292,-0.2004],[294,-0.1945],[296,-0.1934],[298,-0.1944],[300,-0.1991],[302,-0.2014],[304,-0.2062],[306,-0.2168],[308,-0.2233],[310,-0.2287],[312,-0.2334],[314,-0.2445],[316,-0.2541],[318,-0.2643],[320,-0.2673],[322,-0.2645],[324,-0.2541],[328,-0.2571],[330,-0.2576],[332,-0.2231],[334,-0.182],[336,-0.2293],[338,-0.3308],[340,-0.3698],[342,-0.6131],[344,-0.6779],[346,-0.8975],[348,-0.8756],[350,-1.158],[352,-0.9762],[354,-0.9554],[356,-1.1174],[358,-0.8699],[360,-1.0109],[362,-0.8769],[364,-0.2717],[366,0.3376],[368,0.7395]]
       }
       spine_03_ctrl.Scale.X {type=MovieSceneScriptingFloatChannel, default=1, const [1,140]}
       spine_03_ctrl.Scale.Y {type=MovieSceneScriptingFloatChannel, default=1, const [1,140]}
       spine_03_ctrl.Scale.Z {type=MovieSceneScriptingFloatChannel, default=1, const [1,140]}
       spine_04_ctrl.Location.X {type=MovieSceneScriptingFloatChannel, default=0, thinned_from=140, keys [[0,0],[368,0]]}
       spine_04_ctrl.Location.Y {type=MovieSceneScriptingFloatChannel, default=0, thinned_from=140, keys [[0,0],[362,-0.0001],[364,-0.0003],[368,0]]}
       spine_04_ctrl.Location.Z {
        type=MovieSceneScriptingFloatChannel
        default=0
        thinned_from=140
        keys [[0,0],[94,0],[96,-0.0001],[98,0],[110,-0.0005],[112,-0.0004],[114,-0.0006],[116,-0.0002],[118,-0.0006],[120,-0.0001],[124,0],[126,-0.0003],[128,0],[362,-0.0002],[364,-0.0005],[368,0]]
       }
       spine_04_ctrl.Rotation.X {
        type=MovieSceneScriptingFloatChannel
        default=-15.7998
        thinned_from=140
        keys [[0,-16.8641],[100,-16.8167],[102,-16.724],[104,-16.4614],[106,-16.0532],[108,-15.5514],[110,-14.9905],[112,-14.3331],[116,-12.9129],[118,-12.2264],[120,-11.5675],[122,-10.9698],[124,-10.5228],[126,-10.1421],[128,-9.7108],[130,-9.3137],[134,-8.5804],[140,-7.4464],[142,-7.0846],[144,-6.7086],[146,-6.3578],[148,-6.0432],[150,-5.7598],[152,-5.5037],[154,-5.265],[156,-5.0582],[160,-4.6986],[162,-4.5744],[164,-4.4758],[172,-4.1402],[178,-3.9207],[184,-3.7356],[190,-3.5856],[192,-3.5432],[198,-3.4634],[204,-3.4583],[206,-3.4708],[222,-3.4858],[236,-3.4627],[246,-3.4613],[254,-3.4495],[256,-3.4403],[270,-3.4359],[272,-3.4303],[282,-3.4301],[288,-3.4191],[302,-3.4317],[306,-3.4206],[308,-3.4042],[320,-3.372],[334,-3.3706],[336,-3.3854],[340,-3.3435],[342,-3.3035],[348,-3.2488],[352,-3.2491],[360,-3.2733],[362,-3.2856],[364,-4.6773],[366,-12.4237],[368,-15.7998]]
       }
       spine_04_ctrl.Rotation.Y {
        type=MovieSceneScriptingFloatChannel
        default=0.1348
        thinned_from=140
        keys [[0,-0.299],[92,-0.299],[94,-0.303],[96,-0.3094],[98,-0.3226],[100,-0.3282],[102,-0.3247],[104,-0.2991],[106,-0.2505],[108,-0.222],[110,-0.1753],[112,-0.1317],[114,-0.0823],[116,-0.0593],[118,-0.0383],[120,-0.0027],[122,0.0594],[124,0.0831],[126,0.0993],[128,0.0965],[130,0.1018],[132,0.1131],[134,0.1265],[136,0.1247],[138,0.0976],[140,0.0755],[142,0.0847],[144,0.0948],[146,0.0748],[148,0.0505],[150,0.0334],[152,0.017],[154,-0.0028],[156,-0.0126],[158,-0.0193],[160,-0.0496],[162,-0.0674],[164,-0.0864],[166,-0.1057],[168,-0.1162],[170,-0.1285],[172,-0.1549],[174,-0.1706],[176,-0.1666],[178,-0.1597],[180,-0.1634],[182,-0.1726],[184,-0.1797],[186,-0.182],[188,-0.186],[190,-0.1923],[192,-0.1994],[194,-0.2028],[196,-0.2094],[198,-0.2148],[200,-0.2184],[202,-0.2245],[204,-0.2344],[206,-0.2191],[208,-0.2148],[212,-0.209],[216,-0.2054],[218,-0.21],[220,-0.2139],[222,-0.212],[226,-0.2029],[228,-0.201],[230,-0.2021],[232,-0.1951],[234,-0.1886],[236,-0.1896],[238,-0.194],[240,-0.1992],[242,-0.2028],[246,-0.2112],[248,-0.2145],[250,-0.2154],[252,-0.2131],[254,-0.2155],[256,-0.2222],[258,-0.2256],[260,-0.2249],[262,-0.226],[264,-0.2315],[266,-0.2346],[268,-0.2458],[270,-0.2552],[272,-0.2581],[274,-0.2572],[276,-0.2588],[278,-0.2572],[280,-0.2547],[282,-0.2545],[284,-0.2569],[286,-0.2613],[288,-0.2684],[290,-0.2749],[292,-0.28],[294,-0.2823],[298,-0.2807],[300,-0.2788],[302,-0.2742],[306,-0.2606],[308,-0.251],[312,-0.2358],[316,-0.2234],[318,-0.2239],[320,-0.2232],[322,-0.2231],[324,-0.2235],[326,-0.2251],[328,-0.2261],[330,-0.2246],[332,-0.2223],[334,-0.2337],[336,-0.2198],[338,-0.2489],[340,-0.2937],[342,-0.3556],[344,-0.3925],[346,-0.4265],[348,-0.4463],[350,-0.4712],[354,-0.5024],[356,-0.5166],[358,-0.5227],[360,-0.5267],[362,-0.5245],[364,-0.4174],[366,0.0117],[368,0.1348]]
       }
       spine_04_ctrl.Rotation.Z {
        type=MovieSceneScriptingFloatChannel
        default=0.6513
        thinned_from=140
        keys [[0,-0.7433],[92,-0.7433],[94,-0.7383],[96,-0.72],[98,-0.7046],[100,-0.6767],[102,-0.6595],[104,-0.6438],[106,-0.5929],[108,-0.561],[110,-0.5106],[112,-0.4758],[114,-0.4568],[116,-0.4611],[118,-0.3998],[120,-0.2812],[122,-0.1741],[124,-0.1362],[126,-0.111],[128,-0.1592],[130,-0.1858],[132,-0.1777],[134,-0.0872],[136,0.0563],[138,0.1372],[140,0.1837],[142,0.2269],[144,0.2043],[146,0.1352],[148,0.0533],[150,-0.0343],[152,-0.1253],[154,-0.1812],[156,-0.2035],[158,-0.2323],[160,-0.1941],[162,-0.1926],[164,-0.1738],[166,-0.159],[168,-0.1629],[170,-0.1543],[172,-0.1426],[174,-0.1366],[176,-0.1489],[178,-0.1692],[180,-0.1808],[182,-0.1932],[184,-0.2067],[186,-0.2121],[188,-0.2106],[190,-0.1992],[192,-0.1692],[194,-0.1231],[196,-0.0589],[198,0.0044],[200,0.1122],[202,0.1618],[204,0.1697],[206,0.081],[208,0.0863],[210,0.0962],[212,0.1015],[214,0.1],[216,0.0916],[218,0.0869],[220,0.0781],[222,0.0612],[224,0.0245],[226,-0.0078],[228,-0.035],[230,-0.0803],[232,-0.1163],[234,-0.1458],[236,-0.1689],[238,-0.1794],[240,-0.1745],[242,-0.167],[244,-0.1573],[246,-0.1544],[248,-0.1604],[250,-0.1641],[252,-0.1631],[254,-0.159],[256,-0.161],[258,-0.1643],[260,-0.1738],[262,-0.1773],[264,-0.1753],[266,-0.1705],[268,-0.161],[270,-0.164],[272,-0.1563],[274,-0.1388],[276,-0.142],[278,-0.1632],[280,-0.1982],[282,-0.2194],[284,-0.2074],[286,-0.1984],[288,-0.1872],[290,-0.1734],[292,-0.1631],[294,-0.1573],[296,-0.1564],[298,-0.1571],[300,-0.1618],[302,-0.1638],[304,-0.168],[306,-0.1776],[308,-0.1826],[310,-0.1866],[312,-0.19],[314,-0.2009],[316,-0.2101],[318,-0.2197],[320,-0.2218],[322,-0.2187],[324,-0.2171],[326,-0.2166],[330,-0.2171],[332,-0.2065],[334,-0.2073],[336,-0.254],[338,-0.318],[340,-0.3574],[342,-0.4652],[344,-0.5072],[346,-0.551],[348,-0.5963],[350,-0.6286],[352,-0.628],[354,-0.6206],[358,-0.6012],[360,-0.5956],[362,-0.5732],[364,-0.2645],[366,0.2119],[368,0.6513]]
       }
       spine_04_ctrl.Scale.X {type=MovieSceneScriptingFloatChannel, default=1, const [1,140]}
       spine_04_ctrl.Scale.Y {type=MovieSceneScriptingFloatChannel, default=1, const [1,140]}
       spine_04_ctrl.Scale.Z {type=MovieSceneScriptingFloatChannel, default=1, const [1,140]}
       head_ctrl.Location.X {
        type=MovieSceneScriptingFloatChannel
        default=-0.0131
        thinned_from=107
        keys [[0,0],[326,-0.0001],[332,-0.0011],[334,-0.0018],[336,-0.003],[342,-0.0077],[344,-0.0089],[348,-0.0097],[358,-0.0105]]
       }
       head_ctrl.Location.Y {
        type=MovieSceneScriptingFloatChannel
        default=-0.007
        thinned_from=99
        keys [[0,0],[98,0],[108,-0.0001],[326,0],[332,-0.0006],[340,-0.0034],[368,-0.007]]
       }
       head_ctrl.Location.Z {
        type=MovieSceneScriptingFloatChannel
        default=0.0997
        thinned_from=108
        keys [[0,0],[94,0],[96,-0.0002],[98,0],[100,-0.0004],[102,-0.0002],[104,-0.0006],[108,-0.0008],[112,-0.0004],[114,-0.0004],[116,-0.0002],[118,-0.0002],[120,0],[324,0],[326,0.0007],[328,0.0025],[330,0.0051],[332,0.008],[334,0.0135],[336,0.0229],[338,0.0347],[340,0.0472],[342,0.0587],[344,0.0676],[348,0.0741],[354,0.0776],[358,0.0806],[368,0.0997]]
       }
       head_ctrl.Rotation.X {
        type=MovieSceneScriptingFloatChannel
        default=-67.0095
        thinned_from=103
        keys [[0,-92.9563],[92,-92.9563],[98,-90.9668],[100,-89.999],[102,-88.8109],[104,-86.3675],[108,-77.1576],[110,-70.6215],[112,-65.1716],[114,-59.1905],[116,-53.4463],[118,-47.2907],[120,-41.8204],[122,-36.871],[136,-20.1315],[164,-3.6207],[196,0.4371],[198,0.3598],[200,0.2619],[202,0.1765],[204,0.1071],[206,-0.0033],[208,0.0425],[210,0.2011],[212,0.3347],[214,0.3792],[216,0.5193],[218,0.6525],[220,0.7603],[222,0.8358],[224,0.8546],[226,0.8671],[228,0.895],[230,0.98],[232,1.0571],[234,1.1566],[236,1.2521],[238,1.3437],[240,1.4237],[242,1.4796],[244,1.5598],[246,1.6793],[248,1.7606],[250,1.7829],[254,1.7702],[256,1.7506],[258,1.7354],[260,1.714],[262,1.6802],[264,1.5916],[266,1.5213],[268,1.3526],[270,1.3267],[272,1.6297],[274,2.5235],[276,3.8129],[278,4.8802],[280,5.6812],[282,6.9446],[284,7.2595],[286,7.0013],[288,6.6257],[296,4.6408],[302,4.3406],[306,4.1288],[308,4.0829],[310,4.0588],[312,4.0548],[314,4.1144],[316,4.1907],[318,4.2829],[320,4.3392],[322,4.3537],[324,4.3538],[326,4.5041],[328,4.9233],[330,5.4926],[332,6.0874],[334,6.206],[336,5.9605],[338,5.9592],[340,5.9721],[342,5.9099],[344,6.2894],[346,5.8675],[348,4.2166],[350,3.937],[352,3.1379],[354,3.0291],[356,3.2901],[358,2.5795],[360,-5.1163],[362,-21.1789],[364,-40.3072],[366,-57.9772],[368,-67.0095]]
       }
       head_ctrl.Rotation.Y {
        type=MovieSceneScriptingFloatChannel
        default=0.3013
        thinned_from=113
        keys [[0,-6.3242],[92,-6.3242],[94,-6.1787],[96,-6.0871],[98,-6.0522],[100,-5.9493],[102,-5.7886],[104,-5.5014],[106,-4.9394],[108,-4.5455],[110,-3.5174],[112,-2.5884],[114,-1.669],[116,-0.6057],[118,0.5465],[136,5.8742],[164,7.9215],[178,8.5626],[198,8.6415],[200,8.529],[202,8.5163],[212,8.6457],[214,8.653],[216,8.6245],[218,8.7036],[222,8.9076],[224,8.9701],[226,8.987],[228,9.0339],[230,9.1944],[236,9.5466],[240,9.8857],[242,9.9616],[246,10.2681],[248,10.3696],[258,10.2962],[262,10.2211],[264,10.1564],[266,10.1207],[268,10.0368],[270,10.0012],[272,10.167],[274,10.4907],[276,10.9844],[278,11.3459],[280,11.4743],[282,11.4952],[284,11.452],[286,11.3667],[290,11.4586],[292,11.4604],[294,11.4334],[296,11.4311],[298,11.4869],[302,11.5422],[304,11.59],[306,11.7401],[308,11.9835],[312,12.3479],[316,12.4434],[318,12.4446],[322,12.5023],[324,12.5116],[326,12.4504],[328,12.2569],[330,12.0273],[332,11.7606],[334,11.0768],[336,9.8015],[338,8.4745],[340,7.1885],[342,5.9842],[344,5.0475],[346,4.2858],[348,3.6417],[350,3.1467],[352,2.8611],[354,2.6876],[356,2.486],[358,2.3759],[360,1.5061],[362,-0.5197],[364,-0.5595],[366,0],[368,0.3013]]
       }
       head_ctrl.Rotation.Z {
        type=MovieSceneScriptingFloatChannel
        default=6.3778
        thinned_from=115
        keys [[0,-10.6877],[92,-10.6877],[94,-10.805],[96,-10.8067],[98,-10.7419],[100,-10.6064],[102,-10.5893],[104,-10.8202],[106,-10.9345],[108,-11.3586],[110,-11.6201],[112,-12.0473],[114,-12.3058],[116,-12.3984],[118,-12.2524],[120,-11.8967],[122,-11.2418],[136,-10.8518],[164,-14.4289],[178,-14.8985],[184,-15.0359],[190,-15.0407],[192,-14.9953],[198,-14.7374],[204,-14.749],[212,-14.884],[214,-14.9408],[218,-15.2192],[220,-15.3336],[224,-15.4874],[228,-15.7747],[230,-16.0365],[232,-16.2233],[236,-16.695],[238,-16.8885],[240,-17.0178],[248,-17.267],[250,-17.3802],[254,-17.3682],[258,-17.2639],[264,-17.2117],[266,-17.2223],[268,-17.2748],[270,-17.4232],[272,-17.6379],[274,-17.8884],[280,-18.8087],[282,-19.4078],[284,-19.8001],[286,-20.1143],[288,-20.3255],[290,-20.4339],[294,-20.3289],[296,-20.3311],[298,-20.2901],[304,-20.3855],[306,-20.3697],[310,-20.1623],[312,-20.011],[320,-19.603],[324,-19.5181],[326,-17.42],[328,-12.3955],[330,-6.2268],[332,-0.6717],[334,4.5342],[336,10.5422],[338,16.5214],[340,22.2032],[342,27.0241],[344,30.3967],[346,31.7642],[348,31.6896],[350,31.4191],[354,30.6694],[358,29.7368],[360,17.0108],[362,-6.374],[364,-11.8445],[366,0],[368,6.3778]]
       }
       head_ctrl.Scale.X {type=MovieSceneScriptingFloatChannel, default=1, const [1,115]}
       head_ctrl.Scale.Y {type=MovieSceneScriptingFloatChannel, default=1, const [1,115]}
       head_ctrl.Scale.Z {type=MovieSceneScriptingFloatChannel, default=1, const [1,115]}
       clavicle_l_ctrl.Location.X {type=MovieSceneScriptingFloatChannel, default=0, thinned_from=77, keys [[0,0],[368,0]]}
       clavicle_l_ctrl.Location.Y {type=MovieSceneScriptingFloatChannel, default=0, thinned_from=77, keys [[0,0],[360,0],[368,0]]}
       clavicle_l_ctrl.Location.Z {type=MovieSceneScriptingFloatChannel, default=0, thinned_from=77, keys [[0,0],[342,-0.0001],[368,0]]}
       clavicle_l_ctrl.Rotation.X {
        type=MovieSceneScriptingFloatChannel
        default=4.1292
        thinned_from=77
        keys [[0,16.1194],[92,16.1194],[94,16.2994],[96,16.412],[98,16.3675],[100,16.1719],[102,15.8313],[104,15.3517],[114,9.0503],[124,4.4279],[154,1.8897],[174,1.4585],[194,1.4256],[242,0.203],[244,0.2231],[246,0.2288],[248,0.2212],[250,0.2119],[252,0.2045],[254,0.2019],[256,0.205],[258,0.1853],[260,0.1421],[262,0.1073],[264,0.0801],[266,0.0743],[268,0.0898],[270,0.1559],[272,0.2514],[274,0.3068],[276,0.3322],[278,0.3255],[280,0.324],[282,0.3515],[284,0.4123],[286,0.4497],[290,0.4883],[292,0.4778],[294,0.4779],[298,0.4747],[300,0.4688],[302,0.4806],[304,0.4982],[306,0.5318],[308,0.602],[310,0.6292],[312,0.6466],[314,0.6873],[316,0.7854],[318,0.9659],[320,1.2386],[322,1.5754],[324,1.961],[326,2.3861],[328,2.8695],[330,3.3975],[332,3.8311],[334,3.871],[336,3.7287],[338,4.1428],[340,4.713],[342,5.4329],[344,5.6238],[346,5.7783],[348,5.8204],[350,5.6473],[352,5.6106],[354,5.3881],[356,5.209],[358,5.1145],[360,5.0702],[362,4.9957],[364,3.7531],[366,3.5204],[368,4.1292]]
       }
       clavicle_l_ctrl.Rotation.Y {
        type=MovieSceneScriptingFloatChannel
        default=36.106
        thinned_from=77
        keys [[0,26.824],[92,26.824],[94,24.9201],[96,22.5481],[98,19.8827],[102,14.266],[104,11.6122],[114,4.6773],[124,1.8078],[154,0.321],[174,0.0728],[194,0.0579],[242,-0.3518],[244,-0.333],[246,-0.2756],[248,-0.2751],[250,-0.3001],[252,-0.3325],[254,-0.3431],[256,-0.3057],[258,-0.3091],[260,-0.2849],[262,-0.2831],[264,-0.2276],[266,-0.2125],[268,-0.164],[270,-0.1413],[272,-0.0955],[274,-0.0669],[276,-0.0045],[278,0.0022],[280,-0.0162],[282,0.0143],[284,0.0668],[286,0.0632],[288,0.0819],[290,0.0761],[292,0.0535],[294,0.0396],[296,0.0377],[298,0.0631],[300,0.0667],[302,0.0818],[304,0.0938],[306,0.1188],[308,0.1567],[310,0.2098],[312,0.2467],[314,0.2537],[316,0.3274],[318,0.4564],[320,0.636],[322,0.8526],[324,1.1383],[326,1.415],[330,1.9125],[332,2.3716],[334,3.1235],[336,3.2828],[338,3.4186],[340,3.9698],[342,4.0063],[344,3.8977],[346,3.6785],[348,3.2086],[350,2.9033],[352,2.6655],[356,2.3804],[358,2.4294],[360,2.5741],[362,3.4239],[364,26.4021],[366,33.2266],[368,36.106]]
       }
       clavicle_l_ctrl.Rotation.Z {
        type=MovieSceneScriptingFloatChannel
        default=19.0176
        thinned_from=77
        keys [[0,16.0975],[92,16.0975],[94,16.1798],[96,16.2124],[98,16.1567],[100,16.0157],[102,15.7922],[104,15.489],[114,9.4091],[124,4.24],[154,1.274],[174,0.7649],[194,0.7212],[242,-1.3657],[244,-1.3517],[246,-1.3051],[248,-1.3572],[250,-1.4338],[252,-1.4735],[254,-1.4605],[256,-1.4524],[258,-1.4774],[260,-1.5097],[262,-1.5724],[266,-1.5409],[268,-1.4851],[270,-1.3499],[272,-1.2359],[274,-1.1832],[276,-1.1878],[278,-1.2639],[280,-1.3734],[282,-1.4676],[284,-1.5932],[286,-1.5669],[288,-1.4986],[290,-1.4807],[292,-1.4949],[294,-1.5294],[296,-1.5089],[298,-1.4923],[300,-1.4984],[302,-1.5262],[304,-1.5408],[306,-1.5923],[308,-1.6369],[310,-1.6589],[312,-1.7084],[314,-1.7344],[316,-1.6199],[318,-1.2005],[320,-0.5032],[322,0.3822],[324,1.444],[326,2.6169],[328,3.9339],[330,5.3724],[332,6.7068],[334,7.2423],[336,6.7349],[338,5.391],[340,4.0324],[342,2.5652],[344,1.7321],[346,0.7151],[348,-0.1589],[350,-0.4009],[352,-0.0913],[354,0.1466],[356,0.5974],[358,1.1642],[360,2.1518],[362,3.6575],[364,15.2342],[366,16.5321],[368,19.0176]]
       }
       clavicle_l_ctrl.Scale.X {type=MovieSceneScriptingFloatChannel, default=1, const [1,77]}
       clavicle_l_ctrl.Scale.Y {type=MovieSceneScriptingFloatChannel, default=1, const [1,77]}
       clavicle_l_ctrl.Scale.Z {type=MovieSceneScriptingFloatChannel, default=1, const [1,77]}
       clavicle_r_ctrl.Location.X {type=MovieSceneScriptingFloatChannel, default=0, thinned_from=93, keys [[0,0],[360,0],[362,-0.0002],[368,0]]}
       clavicle_r_ctrl.Location.Y {type=MovieSceneScriptingFloatChannel, default=0, thinned_from=93, keys [[0,0],[368,0]]}
       clavicle_r_ctrl.Location.Z {
        type=MovieSceneScriptingFloatChannel
        default=-0.0001
        thinned_from=93
        keys [[0,-0.0001],[334,-0.0001],[340,-0.0003],[360,-0.0001],[362,-0.0003],[368,-0.0001]]
       }
       clavicle_r_ctrl.Rotation.X {
        type=MovieSceneScriptingFloatChannel
        default=-37.7702
        thinned_from=93
        keys [[0,23.463],[92,23.463],[94,23.7237],[96,23.8772],[98,23.7818],[100,23.4456],[102,22.8766],[104,22.0829],[106,21.0724],[108,19.8533],[110,18.4336],[112,16.8814],[116,13.6531],[154,0.2341],[174,0.0741],[176,0.0548],[178,0.0952],[180,0.0811],[182,0.0523],[190,0.0523],[212,0.0532],[214,0.0568],[216,0.0682],[218,0.0975],[220,0.1777],[222,0.3733],[246,4.9448],[248,5.0356],[250,5.0887],[254,5.057],[256,5.0306],[260,5.0343],[262,5.0515],[264,5.0553],[266,4.9133],[268,4.9197],[270,4.9727],[272,5.0536],[274,5.0851],[276,5.0943],[278,5.0808],[280,5.0783],[282,5.1159],[284,5.1782],[286,5.2128],[290,5.2398],[300,5.2186],[304,5.2352],[306,5.268],[308,5.3256],[312,5.3585],[316,5.4293],[320,5.4391],[326,5.3952],[328,5.3935],[330,5.4108],[332,5.3741],[334,5.0667],[336,4.5919],[338,4.771],[340,5.158],[342,5.7999],[344,5.868],[348,6.0871],[350,5.9377],[352,5.8723],[354,5.6237],[356,5.4348],[358,5.3136],[360,5.2399],[362,5.1223],[364,-24.7281],[366,-33.6396],[368,-37.7702]]
       }
       clavicle_r_ctrl.Rotation.Y {
        type=MovieSceneScriptingFloatChannel
        default=34.4896
        thinned_from=93
        keys [[0,29.4447],[92,29.4447],[94,26.5888],[96,23.0227],[98,18.9858],[100,14.7488],[102,10.5827],[104,6.7583],[106,3.5466],[108,1.2185],[110,0.0449],[112,-0.4655],[114,-0.9745],[116,-1.4677],[154,0.2736],[174,0.1377],[176,0.021],[178,-0.0476],[180,-0.0244],[182,-0.0115],[190,-0.0115],[212,-0.0105],[214,-0.008],[216,-0.0001],[218,0.0399],[220,0.1057],[222,0.2067],[246,-0.9481],[248,-1.0829],[250,-0.9724],[252,-0.8121],[254,-0.7425],[256,-0.6493],[258,-0.5953],[260,-0.5904],[262,-0.5838],[264,-0.6474],[266,-0.2547],[268,-0.1531],[270,0.0515],[272,0.4824],[274,1.0253],[276,1.6083],[278,1.9245],[280,2.0372],[282,2.0675],[284,1.9104],[286,1.7304],[288,1.5681],[290,1.4515],[294,1.4097],[296,1.3995],[298,1.3762],[300,1.3838],[302,1.3473],[304,1.3027],[306,1.2504],[308,1.1565],[310,1.0732],[312,1.0024],[314,0.9595],[316,0.9194],[318,0.9102],[320,0.8906],[322,0.9089],[324,0.9175],[326,0.9431],[328,0.9477],[330,0.9159],[332,1.1755],[334,1.981],[336,1.641],[338,0.0533],[340,-1.4932],[342,-3.2908],[344,-3.4278],[346,-3.1424],[348,-3.3104],[350,-3.1534],[352,-3.0549],[354,-3.0584],[356,-3.3077],[358,-3.3476],[360,-3.4186],[362,-3.0571],[364,24.0601],[366,31.5313],[368,34.4896]]
       }
       clavicle_r_ctrl.Rotation.Z {
        type=MovieSceneScriptingFloatChannel
        default=-0.0682
        thinned_from=93
        keys [[0,17.0751],[92,17.0751],[94,16.5121],[96,15.7426],[98,14.7766],[100,13.6759],[104,11.318],[106,10.1845],[108,9.1638],[110,8.3177],[112,7.5612],[114,6.7764],[116,5.9756],[154,1.2206],[174,0.6097],[176,0.5269],[178,0.3716],[180,0.2138],[182,0.1167],[192,0.1165],[212,0.1169],[214,0.1194],[216,0.1296],[218,0.1459],[220,0.213],[222,0.3922],[246,6.4142],[248,7.1145],[250,7.4981],[252,7.7968],[254,7.8907],[256,7.9177],[258,7.9219],[260,7.9457],[262,7.8349],[264,7.8152],[266,7.7555],[268,7.7255],[272,7.6275],[274,7.3942],[276,7.0632],[278,6.9173],[280,6.8238],[282,6.6621],[284,6.6036],[288,6.6401],[290,6.7323],[292,6.8069],[294,6.7573],[298,6.6938],[300,6.6338],[304,6.5646],[306,6.5431],[308,6.5442],[312,6.4771],[316,6.4808],[318,6.4555],[320,6.4129],[324,6.372],[326,6.3406],[328,6.3415],[330,6.3254],[332,6.2426],[334,5.6972],[336,3.5706],[338,1.6309],[340,-0.806],[342,-2.3412],[344,-2.9186],[346,-3.2743],[348,-2.8603],[350,-2.1539],[352,-1.7963],[354,-1.7339],[356,-1.5311],[358,-1.4327],[360,-1.1228],[362,-1.0382],[364,-0.497],[366,0.1154],[368,-0.0682]]
       }
       clavicle_r_ctrl.Scale.X {type=MovieSceneScriptingFloatChannel, default=1, const [1,93]}
       clavicle_r_ctrl.Scale.Y {type=MovieSceneScriptingFloatChannel, default=1, const [1,93]}
       clavicle_r_ctrl.Scale.Z {type=MovieSceneScriptingFloatChannel, default=1, const [1,93]}
       neck_01_ctrl.Location.X {type=MovieSceneScriptingFloatChannel, default=0, thinned_from=81, keys [[0,0],[368,0]]}
       neck_01_ctrl.Location.Y {type=MovieSceneScriptingFloatChannel, default=0, thinned_from=81, keys [[0,0],[364,-0.0002],[368,0]]}
       neck_01_ctrl.Location.Z {
        type=MovieSceneScriptingFloatChannel
        default=0
        thinned_from=81
        keys [[0,0.0001],[94,0],[96,-0.0001],[98,0],[100,-0.0002],[102,0],[110,-0.0006],[112,-0.0004],[114,-0.0007],[116,-0.0003],[118,-0.0007],[120,-0.0001],[124,0],[126,-0.0003],[128,0],[364,-0.0002],[368,0]]
       }
       neck_01_ctrl.Rotation.X {
        type=MovieSceneScriptingFloatChannel
        default=0
        thinned_from=70
        keys [[0,21.7815],[92,21.7815],[94,21.5934],[96,21.0588],[98,20.1291],[100,18.4626],[102,17.375],[104,15.4228],[106,11.0802],[108,8.7124],[110,5.6176],[112,4.7071],[114,4.038],[116,3.9771],[140,3.887],[194,0.012],[214,0.8977],[216,1.1969],[218,1.3346],[220,1.4022],[222,1.3895],[224,1.3853],[226,1.4279],[228,1.4823],[230,1.5008],[232,1.5074],[234,1.5447],[256,2.0644],[272,2.2695],[282,2.657],[284,2.9382],[286,3.0125],[290,3.0581],[292,3.0511],[294,3.0323],[296,3.04],[298,3.0029],[300,3.0053],[302,3.0214],[304,3.044],[306,3.1089],[310,3.2782],[312,3.3374],[314,3.4696],[316,3.5661],[320,3.7193],[328,3.7561],[330,4.355],[332,5.7833],[334,7.3872],[336,8.6424],[338,9.0935],[340,9.1259],[342,9.2869],[344,9.7156],[348,10.4653],[352,10.969],[354,11.1195],[358,11.1696],[360,8.2464],[362,-0.2952],[364,-3.4253],[366,-1.124],[368,0]]
       }
       neck_01_ctrl.Rotation.Y {
        type=MovieSceneScriptingFloatChannel
        default=0
        thinned_from=81
        keys [[0,0.7867],[92,0.7867],[94,0.7141],[96,0.7253],[98,0.707],[100,0.5639],[102,0.419],[104,0.3696],[106,0.3759],[108,0.597],[110,0.6807],[112,0.8256],[114,1.0649],[116,1.1549],[118,1.3518],[120,1.4609],[122,1.5776],[124,1.7777],[126,2.1742],[128,2.0672],[130,2.0488],[132,2.1975],[134,2.0189],[136,1.4858],[140,1.0428],[194,3.8327],[214,2.1584],[216,2.4104],[218,2.5816],[220,2.73],[222,2.8222],[224,2.8694],[226,2.9522],[228,3.1069],[230,3.2549],[232,3.4376],[234,3.604],[256,6.3443],[272,7.409],[282,8.5946],[284,8.4448],[286,8.372],[288,8.3803],[290,8.3501],[292,8.2453],[294,8.201],[296,8.1842],[298,8.142],[302,8.1331],[304,8.179],[306,8.1848],[310,8.3794],[312,8.5173],[314,8.5756],[316,8.6154],[320,8.6584],[328,8.6146],[330,8.2996],[332,7.5081],[334,6.5715],[336,5.4763],[338,4.7295],[340,4.3107],[342,3.918],[344,3.5464],[348,2.8193],[352,2.224],[354,2.038],[358,1.8615],[360,2.2502],[362,4.617],[364,4.3482],[366,1.5219],[368,0]]
       }
       neck_01_ctrl.Rotation.Z {
        type=MovieSceneScriptingFloatChannel
        default=0
        thinned_from=81
        keys [[0,-4.5347],[92,-4.5347],[94,-4.5919],[98,-4.5731],[102,-4.6831],[104,-4.6927],[106,-4.574],[108,-4.4173],[110,-4.1713],[112,-4.0056],[114,-3.7864],[116,-3.5536],[118,-3.2704],[120,-3.0492],[122,-2.8135],[124,-2.7136],[126,-2.5894],[128,-2.6665],[130,-2.699],[132,-2.7427],[134,-2.7639],[136,-2.7638],[138,-2.7356],[140,-2.7209],[194,-3.5122],[216,-4.3132],[222,-4.4078],[224,-4.4548],[226,-4.5183],[228,-4.5704],[230,-4.674],[232,-4.7314],[234,-4.8174],[256,-5.6479],[272,-5.9725],[282,-6.3765],[284,-6.3778],[286,-6.3383],[288,-6.2825],[290,-6.254],[294,-6.2466],[298,-6.2576],[306,-6.2856],[310,-6.1876],[312,-6.0994],[316,-5.9966],[320,-5.9246],[328,-5.8988],[330,-6.0471],[332,-6.4333],[334,-6.8762],[336,-7.293],[338,-7.6368],[340,-7.828],[342,-7.9674],[344,-8.0845],[348,-8.3436],[354,-8.8195],[358,-9.2212],[360,-4.0236],[362,-5.4336],[364,-4.5399],[366,-1.5785],[368,0]]
       }
       neck_01_ctrl.Scale.X {type=MovieSceneScriptingFloatChannel, default=1, const [1,81]}
       neck_01_ctrl.Scale.Y {type=MovieSceneScriptingFloatChannel, default=1, const [1,81]}
       neck_01_ctrl.Scale.Z {type=MovieSceneScriptingFloatChannel, default=1, const [1,81]}
       neck_02_ctrl.Location.X {type=MovieSceneScriptingFloatChannel, default=0, thinned_from=81, keys [[0,0],[368,0]]}
       neck_02_ctrl.Location.Y {
        type=MovieSceneScriptingFloatChannel
        default=0
        thinned_from=81
        keys [[0,0],[102,0],[108,-0.0002],[140,0],[328,0],[330,-0.0069],[332,-0.0239],[334,-0.0452],[336,-0.0654],[338,-0.0787],[342,-0.0947],[344,-0.1023],[348,-0.1161],[352,-0.1268],[354,-0.1304],[358,-0.1336],[360,0.0073],[362,0.0736],[364,0.0801],[366,0.028],[368,0]]
       }
       neck_02_ctrl.Location.Z {
        type=MovieSceneScriptingFloatChannel
        default=0
        thinned_from=81
        keys [[0,0],[94,0],[96,-0.0002],[98,0],[100,-0.0003],[102,-0.0002],[104,-0.0004],[108,-0.0007],[112,-0.0004],[114,-0.0004],[116,-0.0002],[118,-0.0003],[120,0],[124,0],[126,-0.0002],[128,0],[328,0],[330,-0.0205],[332,-0.0708],[334,-0.1345],[336,-0.1948],[338,-0.2351],[340,-0.2598],[342,-0.2839],[344,-0.3069],[348,-0.3479],[352,-0.3789],[354,-0.3896],[358,-0.3987],[360,0.0218],[362,0.2198],[364,0.2393],[366,0.0838],[368,0]]
       }
       neck_02_ctrl.Rotation.X {
        type=MovieSceneScriptingFloatChannel
        default=0
        thinned_from=70
        keys [[0,-1.5481],[92,-1.5481],[94,-1.4679],[96,-1.4356],[98,-1.4608],[100,-1.5293],[102,-1.547],[104,-1.5471],[106,-1.4777],[108,-1.1781],[110,-0.5027],[112,-0.0117],[114,0.587],[116,1.1444],[140,5.1399],[194,5.8924],[214,6.09],[216,6.2369],[218,6.3157],[220,6.3753],[222,6.409],[226,6.4247],[228,6.4427],[232,6.5238],[234,6.5759],[256,8.1412],[272,8.7693],[282,9.7723],[284,9.8214],[290,9.5893],[294,9.474],[296,9.4621],[304,9.3443],[308,9.2663],[316,9.2964],[320,9.3422],[328,9.3421],[330,10.1788],[332,12.2407],[334,14.5019],[336,16.145],[338,16.633],[340,16.3346],[342,16.0755],[344,16.2078],[348,16.5781],[352,17.1727],[354,17.4435],[358,17.575],[360,10.8228],[362,2.3587],[364,-1.0454],[366,-0.3659],[368,0]]
       }
       neck_02_ctrl.Rotation.Y {
        type=MovieSceneScriptingFloatChannel
        default=0
        thinned_from=81
        keys [[0,1.4778],[92,1.4778],[94,1.5314],[96,1.6086],[98,1.7118],[100,1.8675],[102,1.9714],[104,2.0984],[106,2.354],[108,2.5483],[110,2.9316],[112,3.0792],[114,3.0972],[116,3.1686],[118,3.3078],[120,3.33],[122,3.1684],[124,3.1037],[126,3.0668],[130,2.9391],[132,2.8835],[134,2.8613],[136,2.7626],[138,2.7407],[140,2.7779],[194,5.0507],[214,4.7285],[216,4.7229],[222,4.8583],[224,4.8739],[226,4.8725],[228,4.8823],[230,4.9415],[234,5.0374],[256,6.0997],[272,6.517],[282,7.0507],[284,6.9735],[286,6.9201],[296,6.9132],[300,6.9356],[304,6.9227],[312,6.9969],[320,6.9118],[326,6.9029],[328,6.8753],[330,6.6606],[332,6.1854],[334,5.5751],[336,4.8479],[338,4.3311],[340,3.9502],[342,3.591],[344,3.2867],[348,2.6968],[352,2.2997],[354,2.1708],[358,1.9525],[360,1.0616],[362,0.8886],[364,0.564],[366,0.1858],[368,0]]
       }
       neck_02_ctrl.Rotation.Z {
        type=MovieSceneScriptingFloatChannel
        default=0
        thinned_from=81
        keys [[0,-4.7326],[92,-4.7326],[94,-4.7417],[96,-4.7387],[102,-4.6623],[104,-4.614],[106,-4.4264],[108,-4.3382],[110,-4.0805],[112,-3.9404],[114,-3.7416],[116,-3.4565],[118,-3.2509],[120,-3.2105],[122,-3.2109],[124,-3.2633],[126,-3.3284],[128,-3.3189],[130,-3.3285],[132,-3.4488],[136,-3.6596],[138,-3.7029],[140,-3.6679],[194,-3.8318],[214,-4.4129],[216,-4.5145],[218,-4.5494],[220,-4.571],[224,-4.5738],[228,-4.6361],[230,-4.6826],[232,-4.7122],[234,-4.7695],[256,-5.1535],[272,-5.3012],[282,-5.4678],[284,-5.4702],[286,-5.4382],[288,-5.3899],[292,-5.3391],[296,-5.3427],[298,-5.3256],[304,-5.3408],[306,-5.3217],[310,-5.2283],[312,-5.1584],[320,-4.9945],[326,-4.9704],[328,-4.9713],[330,-4.6399],[332,-3.8337],[336,-1.706],[338,-0.9718],[340,-0.4455],[342,0.0697],[344,0.5139],[348,1.2903],[352,1.8282],[354,2.0172],[358,2.1817],[360,-4.391],[362,-11.4209],[364,-10.9163],[366,-3.8207],[368,0]]
       }
       neck_02_ctrl.Scale.X {type=MovieSceneScriptingFloatChannel, default=1, const [1,81]}
       neck_02_ctrl.Scale.Y {type=MovieSceneScriptingFloatChannel, default=1, const [1,81]}
       neck_02_ctrl.Scale.Z {type=MovieSceneScriptingFloatChannel, default=1, const [1,81]}
       head_ik_ctrl.Location.X {
        type=MovieSceneScriptingFloatChannel
        default=-0.0131
        thinned_from=140
        keys [[0,0.5145],[92,0.5145],[94,0.5232],[96,0.4995],[98,0.4692],[100,0.4377],[102,0.4268],[104,0.3758],[106,0.2371],[108,0.1237],[110,-0.0007],[112,-0.062],[114,-0.1238],[116,-0.157],[118,-0.2065],[120,-0.2336],[122,-0.2557],[124,-0.2849],[126,-0.346],[128,-0.3109],[130,-0.2939],[132,-0.304],[134,-0.2772],[136,-0.1948],[138,-0.1658],[140,-0.1418],[142,-0.1339],[144,-0.2117],[146,-0.3403],[148,-0.444],[150,-0.546],[152,-0.6551],[154,-0.7504],[156,-0.8093],[158,-0.8626],[160,-0.9281],[162,-0.99],[164,-1.0284],[166,-1.043],[168,-1.0382],[170,-0.9952],[172,-0.9353],[174,-0.9064],[176,-0.8821],[178,-0.8388],[180,-0.8146],[182,-0.7995],[184,-0.7505],[186,-0.74],[188,-0.7325],[190,-0.7353],[196,-0.7692],[198,-0.767],[200,-0.7437],[202,-0.7232],[204,-0.6658],[206,-0.5968],[208,-0.5224],[210,-0.4492],[212,-0.4016],[214,-0.3913],[216,-0.4179],[218,-0.4706],[220,-0.5181],[222,-0.5498],[224,-0.56],[226,-0.573],[228,-0.6127],[230,-0.6498],[232,-0.7041],[234,-0.747],[236,-0.7986],[238,-0.8449],[240,-0.893],[242,-0.9513],[244,-1.0166],[246,-1.0727],[248,-1.1112],[250,-1.1437],[252,-1.1491],[254,-1.1684],[256,-1.1967],[258,-1.2191],[260,-1.2279],[262,-1.2277],[266,-1.2361],[268,-1.2571],[270,-1.2543],[272,-1.2274],[274,-1.2161],[276,-1.261],[278,-1.3009],[280,-1.312],[282,-1.2871],[284,-1.2477],[286,-1.2378],[288,-1.2473],[290,-1.2498],[292,-1.2367],[298,-1.2273],[302,-1.225],[304,-1.2298],[306,-1.2318],[310,-1.2695],[312,-1.2979],[314,-1.3066],[322,-1.3147],[328,-1.3071],[330,-1.2],[332,-0.9463],[334,-0.6659],[336,-0.3758],[338,-0.1923],[340,-0.1117],[342,-0.002],[344,0.0865],[346,-0.0156],[348,0.2787],[350,0.1765],[352,0.4311],[354,0.4738],[356,0.3048],[358,0.5354],[360,0.2434],[362,-0.2551],[364,-0.4234],[366,-0.125],[368,-0.0131]]
       }
       head_ik_ctrl.Location.Y {
        type=MovieSceneScriptingFloatChannel
        default=-0.007
        thinned_from=140
        keys [[0,2.6516],[92,2.6516],[94,2.6315],[96,2.5616],[98,2.4287],[100,2.1892],[102,2.0488],[104,1.8151],[106,1.2808],[108,1.0339],[110,0.7099],[112,0.6545],[114,0.6361],[116,0.6956],[118,0.8616],[120,1.0105],[122,1.0786],[124,1.1585],[126,1.2499],[128,1.2878],[130,1.3393],[132,1.3806],[134,1.397],[136,1.4059],[138,1.4211],[140,1.3989],[142,1.3464],[144,1.2605],[146,1.26],[148,1.2708],[150,1.3071],[152,1.3742],[154,1.3983],[156,1.4086],[158,1.3942],[160,1.4346],[162,1.4224],[164,1.3824],[166,1.3369],[168,1.307],[170,1.2801],[172,1.2411],[174,1.1702],[176,1.108],[178,1.0405],[182,0.9153],[184,0.8542],[188,0.7498],[190,0.7081],[192,0.6783],[194,0.6547],[196,0.6414],[198,0.623],[200,0.6318],[202,0.6221],[204,0.5974],[206,0.5225],[208,0.5741],[210,0.682],[212,0.766],[214,0.7964],[216,0.8653],[218,0.9363],[220,0.979],[222,0.9865],[224,0.9901],[226,1.0127],[228,1.0436],[230,1.0708],[232,1.0919],[234,1.1277],[236,1.1696],[238,1.1955],[240,1.1965],[242,1.2092],[244,1.219],[248,1.2594],[250,1.2662],[252,1.2702],[254,1.2812],[256,1.2749],[258,1.2727],[262,1.2754],[264,1.2647],[266,1.2575],[268,1.275],[270,1.3374],[272,1.4639],[274,1.585],[276,1.6368],[278,1.6186],[280,1.6102],[282,1.6509],[284,1.6961],[286,1.6967],[290,1.6862],[294,1.6663],[296,1.6663],[298,1.6565],[302,1.655],[304,1.6584],[306,1.6661],[312,1.7085],[314,1.7317],[316,1.75],[320,1.7792],[328,1.782],[330,1.94],[332,2.3149],[334,2.7184],[336,3.0012],[338,3.0617],[340,2.9971],[342,2.945],[344,2.961],[346,3.1553],[348,2.9683],[350,3.2158],[352,2.9839],[354,3.0041],[356,3.3046],[358,3.114],[360,2.382],[362,0.2668],[364,-0.6467],[366,-0.2407],[368,-0.007]]
       }
       head_ik_ctrl.Location.Z {
        type=MovieSceneScriptingFloatChannel
        default=0.0997
        thinned_from=140
        keys [[0,2.4227],[92,2.4227],[94,2.4071],[96,2.3569],[98,2.2684],[100,2.101],[102,1.9767],[104,1.7422],[106,1.2462],[108,0.9699],[110,0.6428],[112,0.5449],[114,0.4788],[116,0.4699],[118,0.5123],[120,0.5303],[122,0.5047],[124,0.4915],[126,0.4812],[128,0.451],[130,0.4237],[132,0.3935],[134,0.3593],[136,0.3234],[138,0.2868],[142,0.2047],[144,0.1608],[146,0.1242],[148,0.0909],[150,0.0589],[152,0.0251],[154,-0.0067],[156,-0.033],[158,-0.0549],[160,-0.0817],[162,-0.0985],[164,-0.1068],[168,-0.1133],[170,-0.1132],[172,-0.1086],[174,-0.1023],[176,-0.0988],[178,-0.0926],[182,-0.0823],[184,-0.0748],[188,-0.0669],[190,-0.0645],[192,-0.0637],[194,-0.0631],[196,-0.0635],[198,-0.0622],[200,-0.0617],[202,-0.0592],[204,-0.0532],[206,-0.041],[208,-0.0432],[210,-0.053],[212,-0.0619],[214,-0.0656],[216,-0.0756],[218,-0.0873],[220,-0.0948],[222,-0.0975],[224,-0.099],[226,-0.1033],[228,-0.1102],[230,-0.1165],[232,-0.1239],[234,-0.1332],[236,-0.1439],[238,-0.1518],[240,-0.156],[242,-0.1627],[244,-0.17],[246,-0.1794],[248,-0.1868],[250,-0.1914],[252,-0.1925],[254,-0.1967],[258,-0.2009],[260,-0.2018],[262,-0.2017],[264,-0.2],[266,-0.199],[268,-0.203],[270,-0.2143],[272,-0.2368],[274,-0.2626],[276,-0.2815],[278,-0.284],[280,-0.285],[282,-0.292],[284,-0.298],[288,-0.2966],[290,-0.2945],[294,-0.2868],[296,-0.2863],[298,-0.2838],[302,-0.283],[306,-0.287],[308,-0.2963],[310,-0.3024],[312,-0.3119],[314,-0.3186],[316,-0.3259],[320,-0.3341],[322,-0.3348],[326,-0.3337],[328,-0.3309],[330,-0.381],[332,-0.5258],[334,-0.7162],[336,-0.8872],[338,-0.9974],[340,-1.056],[342,-1.1471],[344,-1.2743],[346,-1.5546],[348,-1.5407],[350,-1.7941],[352,-1.7393],[354,-1.7875],[356,-1.876],[358,-1.6179],[360,-0.8951],[362,-0.0734],[364,0.0042],[366,0.012],[368,0.0997]]
       }
       head_ik_ctrl.Rotation.X {
        type=MovieSceneScriptingFloatChannel
        default=-67.0095
        thinned_from=140
        keys [[0,-92.9563],[92,-92.9563],[98,-90.9668],[100,-89.999],[102,-88.8109],[104,-86.3675],[108,-77.1576],[110,-70.6215],[112,-65.1716],[114,-59.1905],[116,-53.4463],[118,-47.2907],[120,-41.8204],[122,-36.871],[124,-33.6313],[126,-29.7596],[128,-26.2242],[130,-23.6426],[132,-20.8855],[134,-18.5774],[136,-16.3588],[138,-14.3348],[140,-12.4245],[142,-10.8002],[144,-9.0204],[146,-7.1784],[148,-5.7666],[150,-4.6012],[152,-3.5078],[154,-2.8002],[156,-2.456],[158,-2.3301],[160,-2.045],[162,-2.0477],[164,-2.2999],[166,-2.7056],[168,-3.1193],[170,-3.3367],[172,-3.3634],[178,-3.9298],[180,-4.0448],[182,-4.1714],[184,-4.3297],[186,-4.4369],[188,-4.6092],[194,-5.08],[198,-5.4425],[200,-5.675],[202,-5.8779],[204,-6.0427],[206,-6.3049],[208,-6.196],[210,-5.8194],[212,-5.5023],[214,-5.3965],[216,-5.0638],[218,-4.7473],[220,-4.4914],[222,-4.3121],[224,-4.2674],[226,-4.2376],[228,-4.1714],[230,-3.9697],[232,-3.7864],[234,-3.55],[236,-3.3234],[238,-3.1058],[240,-2.9158],[242,-2.783],[244,-2.5925],[246,-2.3086],[248,-2.1156],[250,-2.0626],[252,-2.0802],[254,-2.0927],[256,-2.1394],[258,-2.1755],[260,-2.2264],[262,-2.3067],[264,-2.517],[266,-2.6839],[268,-3.0846],[270,-3.1462],[272,-2.309],[274,-0.7894],[276,1.403],[278,3.2178],[280,4.5797],[282,5.4842],[284,5.7096],[286,5.5247],[288,5.2559],[290,5.0432],[292,4.8773],[294,4.694],[296,4.6408],[302,4.3406],[306,4.1288],[308,4.0829],[310,4.0588],[312,4.0548],[314,4.1144],[316,4.1907],[318,4.2829],[320,4.3392],[322,4.3537],[324,4.3538],[326,4.5041],[328,4.9233],[330,5.4926],[332,6.0874],[334,6.206],[336,5.9605],[338,5.9592],[340,5.9721],[342,5.9099],[344,6.2894],[346,5.8675],[348,4.2166],[350,3.937],[352,3.1379],[354,3.0291],[356,3.2901],[358,2.5795],[360,-5.1163],[362,-21.1789],[364,-40.3072],[366,-57.9772],[368,-67.0095]]
       }
       head_ik_ctrl.Rotation.Y {
        type=MovieSceneScriptingFloatChannel
        default=0.3013
        thinned_from=140
        keys [[0,-6.3242],[92,-6.3242],[94,-6.1787],[96,-6.0871],[98,-6.0522],[100,-5.9493],[102,-5.7886],[104,-5.5014],[106,-4.9394],[108,-4.5455],[110,-3.5174],[112,-2.5884],[114,-1.669],[116,-0.6057],[118,0.5465],[120,1.2509],[122,1.4536],[124,1.6382],[126,1.809],[128,1.9701],[134,2.4383],[136,2.6034],[138,2.7802],[142,3.1657],[144,3.3412],[146,3.4999],[148,3.6424],[150,3.7695],[152,3.8818],[154,3.9799],[156,4.0645],[158,4.1363],[160,4.1959],[174,4.4885],[182,4.62],[190,4.7107],[196,4.7435],[198,4.7464],[200,4.4793],[202,4.4491],[204,4.5101],[206,4.5501],[210,4.7046],[212,4.7563],[214,4.7737],[216,4.7061],[218,4.8938],[220,5.1518],[222,5.3783],[224,5.5268],[226,5.5671],[228,5.6783],[230,6.0596],[234,6.5962],[236,6.8961],[238,7.3187],[240,7.7014],[242,7.8817],[244,8.231],[246,8.6095],[248,8.8507],[250,8.8387],[258,8.6762],[262,8.4979],[264,8.3442],[266,8.2595],[268,8.0603],[270,7.9756],[272,8.3693],[274,9.1383],[276,10.3108],[278,11.1695],[280,11.4743],[282,11.4952],[284,11.452],[286,11.3667],[290,11.4586],[292,11.4604],[294,11.4334],[296,11.4311],[298,11.4869],[302,11.5422],[304,11.59],[306,11.7401],[308,11.9835],[312,12.3479],[316,12.4434],[318,12.4446],[322,12.5023],[324,12.5116],[326,12.4504],[328,12.2569],[330,12.0273],[332,11.7606],[334,11.0768],[336,9.8015],[338,8.4745],[340,7.1885],[342,5.9842],[344,5.0475],[346,4.2858],[348,3.6417],[350,3.1467],[352,2.8611],[354,2.6876],[356,2.486],[358,2.3759],[360,1.5061],[362,-0.5197],[364,-0.5595],[366,0],[368,0.3013]]
       }
       head_ik_ctrl.Rotation.Z {
        type=MovieSceneScriptingFloatChannel
        default=6.3778
        thinned_from=140
        keys [[0,-10.6877],[92,-10.6877],[94,-10.805],[96,-10.8067],[98,-10.7419],[100,-10.6064],[102,-10.5893],[104,-10.8202],[106,-10.9345],[108,-11.3586],[110,-11.6201],[112,-12.0473],[114,-12.3058],[116,-12.3984],[118,-12.2524],[120,-11.8967],[122,-11.2418],[126,-10.3426],[128,-10.0722],[130,-9.7124],[132,-9.5571],[134,-8.9707],[136,-8.0576],[138,-7.4517],[140,-6.9683],[142,-6.5017],[144,-6.2844],[146,-6.4171],[148,-6.7483],[150,-7.1789],[152,-7.5488],[154,-7.8199],[158,-7.8824],[160,-7.9928],[162,-8.1423],[164,-8.2245],[166,-8.3738],[168,-8.6565],[170,-8.8781],[172,-9.1525],[174,-9.2517],[178,-9.5219],[180,-9.6002],[182,-9.7092],[184,-9.8483],[186,-9.8459],[188,-9.8759],[190,-9.8597],[192,-9.7518],[194,-9.5645],[196,-9.3424],[198,-9.1393],[202,-9.1304],[204,-9.1669],[206,-9.2925],[212,-9.4875],[214,-9.6225],[218,-10.2836],[220,-10.5552],[224,-10.9205],[226,-11.2454],[228,-11.6028],[230,-12.2247],[232,-12.6683],[234,-13.2491],[236,-13.7886],[238,-14.2482],[240,-14.5553],[244,-14.8647],[246,-14.9807],[248,-15.1472],[250,-15.4159],[254,-15.3875],[256,-15.2311],[258,-15.1397],[264,-15.0157],[266,-15.041],[268,-15.1655],[270,-15.5182],[272,-16.0281],[274,-16.6229],[276,-17.3225],[280,-18.8087],[282,-19.4078],[284,-19.8001],[286,-20.1143],[288,-20.3255],[290,-20.4339],[294,-20.3289],[296,-20.3311],[298,-20.2901],[304,-20.3855],[306,-20.3697],[310,-20.1623],[312,-20.011],[320,-19.603],[324,-19.5181],[326,-17.42],[328,-12.3955],[330,-6.2268],[332,-0.6717],[334,4.5342],[336,10.5422],[338,16.5214],[340,22.2032],[342,27.0241],[344,30.3967],[346,31.7642],[348,31.6896],[350,31.4191],[354,30.6694],[358,29.7368],[360,17.0108],[362,-6.374],[364,-11.8445],[366,0],[368,6.3778]]
       }
       head_ik_ctrl.Scale.X {type=MovieSceneScriptingFloatChannel, default=1, const [1,140]}
       head_ik_ctrl.Scale.Y {type=MovieSceneScriptingFloatChannel, default=1, const [1,140]}
       head_ik_ctrl.Scale.Z {type=MovieSceneScriptingFloatChannel, default=1, const [1,140]}
       chest_ctrl.Location.X {
        type=MovieSceneScriptingFloatChannel
        default=0.0834
        thinned_from=140
        keys [[0,-0.28],[92,-0.28],[94,-0.2778],[96,-0.2704],[98,-0.262],[100,-0.2501],[104,-0.2348],[106,-0.2151],[108,-0.1977],[110,-0.1741],[112,-0.1602],[114,-0.1527],[116,-0.1471],[118,-0.1268],[120,-0.1015],[122,-0.0897],[126,-0.0807],[128,-0.085],[130,-0.0883],[132,-0.0868],[134,-0.0737],[136,-0.0484],[138,-0.0258],[140,-0.0111],[142,-0.0078],[144,-0.0145],[146,-0.0172],[148,-0.0188],[152,-0.0273],[154,-0.0257],[156,-0.0235],[158,-0.0232],[160,-0.007],[162,0.0001],[164,0.0093],[166,0.0181],[168,0.0219],[170,0.0277],[172,0.0388],[174,0.0456],[176,0.0435],[178,0.0395],[180,0.0403],[182,0.0429],[184,0.0447],[186,0.0454],[188,0.0474],[190,0.051],[194,0.0621],[196,0.0703],[198,0.078],[200,0.0889],[202,0.0956],[204,0.0956],[206,0.0857],[208,0.0857],[210,0.0843],[216,0.0843],[218,0.0832],[220,0.0832],[222,0.0811],[224,0.0762],[226,0.0717],[228,0.0686],[230,0.0649],[232,0.0592],[234,0.0542],[236,0.0525],[238,0.0532],[242,0.0532],[244,0.0599],[248,0.0599],[250,0.0612],[254,0.0612],[256,0.0648],[258,0.0648],[260,0.0645],[266,0.0645],[268,0.073],[270,0.073],[272,0.0782],[274,0.0795],[276,0.0795],[278,0.0774],[280,0.0733],[282,0.0713],[284,0.0713],[286,0.0754],[288,0.0754],[290,0.0825],[292,0.0825],[294,0.0868],[302,0.0868],[304,0.08],[306,0.08],[308,0.0722],[310,0.0722],[312,0.0656],[314,0.0656],[316,0.0592],[318,0.0592],[320,0.0579],[322,0.0579],[324,0.0586],[326,0.0586],[328,0.0617],[330,0.0617],[332,0.0599],[334,0.0626],[336,0.0181],[338,0.0127],[340,0.025],[342,0.0269],[344,0.0311],[346,0.0288],[348,0.0177],[350,0.0095],[352,0.0195],[354,0.0316],[356,0.0445],[358,0.0595],[360,0.0681],[362,0.0866],[364,0.1028],[366,0.0649],[368,0.0834]]
       }
       chest_ctrl.Location.Y {
        type=MovieSceneScriptingFloatChannel
        default=-2.5968
        thinned_from=140
        keys [[0,-7.8173],[100,-7.8027],[102,-7.76],[104,-7.6388],[106,-7.4499],[108,-7.2086],[110,-6.9295],[112,-6.6024],[118,-5.5158],[120,-5.1679],[122,-4.8513],[124,-4.6044],[126,-4.3907],[128,-4.157],[132,-3.7169],[134,-3.5033],[140,-2.8442],[144,-2.4174],[146,-2.2146],[148,-2.0282],[150,-1.8571],[152,-1.7002],[154,-1.5547],[156,-1.4266],[158,-1.3134],[160,-1.2092],[162,-1.1317],[164,-1.0715],[168,-0.9713],[170,-0.9194],[172,-0.8701],[176,-0.7789],[178,-0.7355],[180,-0.6952],[182,-0.657],[184,-0.6211],[186,-0.5882],[188,-0.5586],[190,-0.5306],[192,-0.5063],[194,-0.4883],[196,-0.4736],[198,-0.462],[200,-0.4605],[204,-0.4609],[206,-0.4642],[208,-0.4642],[210,-0.4657],[216,-0.4657],[220,-0.4704],[222,-0.4698],[228,-0.465],[230,-0.4644],[232,-0.4616],[234,-0.46],[236,-0.4594],[242,-0.4598],[244,-0.4588],[248,-0.4588],[250,-0.4567],[254,-0.4567],[256,-0.4508],[258,-0.4508],[260,-0.4517],[266,-0.4517],[268,-0.4499],[270,-0.4499],[272,-0.4472],[274,-0.4481],[278,-0.448],[284,-0.4468],[286,-0.4434],[292,-0.4436],[294,-0.4456],[302,-0.4456],[304,-0.4468],[306,-0.4468],[308,-0.4373],[310,-0.4373],[312,-0.4312],[314,-0.4312],[316,-0.4265],[318,-0.4265],[320,-0.4242],[326,-0.4245],[328,-0.4253],[330,-0.4253],[332,-0.4231],[334,-0.4296],[336,-0.4205],[338,-0.3181],[340,-0.1753],[342,0.0128],[344,0.1848],[346,0.3509],[348,0.5143],[350,0.615],[352,0.7021],[354,0.7197],[356,0.6001],[358,0.3613],[360,-0.0101],[362,-0.4515],[364,-1.33],[366,-2.3694],[368,-2.5968]]
       }
       chest_ctrl.Location.Z {
        type=MovieSceneScriptingFloatChannel
        default=-0.5428
        thinned_from=140
        keys [[0,-3.4586],[92,-3.4586],[100,-3.4447],[102,-3.4063],[104,-3.3],[106,-3.1387],[108,-2.9416],[110,-2.7248],[112,-2.4841],[114,-2.2363],[116,-2.0016],[118,-1.7853],[120,-1.5891],[122,-1.4229],[124,-1.3005],[126,-1.2005],[128,-1.0943],[130,-0.9985],[132,-0.9103],[136,-0.7451],[138,-0.6683],[140,-0.5962],[142,-0.5298],[144,-0.4662],[146,-0.4097],[148,-0.3607],[150,-0.3182],[152,-0.2813],[154,-0.2488],[156,-0.2216],[158,-0.1985],[160,-0.1781],[162,-0.1636],[164,-0.1525],[170,-0.1259],[172,-0.1177],[176,-0.1029],[178,-0.0961],[180,-0.0899],[182,-0.0842],[184,-0.0789],[186,-0.0741],[188,-0.0699],[190,-0.0659],[192,-0.0626],[194,-0.0601],[196,-0.0581],[198,-0.0566],[200,-0.0565],[204,-0.0566],[206,-0.057],[208,-0.057],[210,-0.0572],[216,-0.0572],[220,-0.0578],[230,-0.0569],[234,-0.0562],[248,-0.0561],[250,-0.0558],[254,-0.0558],[256,-0.055],[266,-0.0551],[270,-0.0549],[272,-0.0546],[278,-0.0547],[284,-0.0545],[286,-0.054],[292,-0.0541],[294,-0.0544],[306,-0.0545],[308,-0.0532],[310,-0.0532],[312,-0.0523],[314,-0.0523],[316,-0.0516],[318,-0.0516],[320,-0.0513],[332,-0.0515],[334,-0.0521],[336,-0.082],[338,-0.1524],[340,-0.2584],[342,-0.3869],[344,-0.5322],[346,-0.6803],[348,-0.818],[350,-0.9341],[352,-1.0131],[354,-1.043],[356,-0.9351],[358,-0.6799],[360,-0.3888],[362,-0.171],[364,-0.2094],[366,-0.4727],[368,-0.5428]]
       }
       chest_ctrl.Rotation.X {
        type=MovieSceneScriptingFloatChannel
        default=-21.6411
        thinned_from=140
        keys [[0,-51.4754],[94,-51.4465],[100,-51.3279],[102,-51.0346],[104,-50.2055],[106,-48.9232],[108,-47.336],[110,-45.5616],[112,-43.4797],[116,-38.9781],[118,-36.7994],[120,-34.7135],[122,-32.8264],[124,-31.4065],[126,-30.1961],[128,-28.8335],[130,-27.5731],[134,-25.2321],[140,-21.619],[144,-19.2787],[146,-18.1713],[148,-17.1744],[150,-16.274],[152,-15.4578],[154,-14.6969],[156,-14.0363],[160,-12.8913],[162,-12.4932],[164,-12.1783],[172,-11.1108],[178,-10.4116],[182,-10.0068],[186,-9.6495],[190,-9.3444],[192,-9.2119],[196,-9.0374],[198,-8.9682],[204,-8.9691],[206,-8.9933],[218,-9.0162],[220,-9.0421],[230,-8.9937],[236,-8.9551],[254,-8.9363],[256,-8.8874],[266,-8.8947],[270,-8.8802],[272,-8.8581],[284,-8.8549],[286,-8.8284],[292,-8.8299],[306,-8.8551],[308,-8.7814],[310,-8.7814],[312,-8.7351],[314,-8.7351],[316,-8.6984],[330,-8.6872],[332,-8.6716],[334,-8.721],[336,-8.6375],[338,-7.7925],[340,-6.5895],[342,-4.9807],[344,-3.4477],[346,-1.8455],[348,-0.3943],[350,0.7116],[352,1.4619],[354,1.6581],[356,0.5926],[358,-1.8372],[360,-5.0018],[362,-8.3252],[364,-13.7464],[366,-20.3835],[368,-21.6411]]
       }
       chest_ctrl.Rotation.Y {
        type=MovieSceneScriptingFloatChannel
        default=0.4703
        thinned_from=140
        keys [[0,-0.9112],[92,-0.9112],[94,-0.9236],[96,-0.9463],[98,-0.9892],[100,-1.0111],[102,-1.002],[104,-0.9349],[106,-0.8035],[108,-0.7166],[110,-0.5882],[112,-0.4532],[114,-0.3098],[116,-0.2465],[118,-0.1699],[120,-0.0267],[122,0.1888],[124,0.2726],[126,0.3269],[128,0.3053],[130,0.3165],[132,0.3536],[134,0.4178],[136,0.4481],[138,0.3827],[140,0.3198],[142,0.3466],[144,0.3634],[146,0.285],[148,0.1972],[150,0.1343],[152,0.0761],[154,0.0126],[156,-0.0176],[158,-0.0383],[160,-0.1299],[162,-0.1846],[164,-0.2427],[166,-0.3018],[168,-0.3332],[170,-0.3709],[172,-0.4508],[174,-0.4985],[176,-0.4859],[178,-0.4636],[180,-0.4736],[182,-0.4998],[184,-0.5196],[186,-0.5256],[188,-0.5373],[190,-0.5568],[192,-0.5805],[194,-0.5957],[196,-0.622],[200,-0.6677],[202,-0.6916],[204,-0.6916],[206,-0.6677],[208,-0.6677],[210,-0.6474],[216,-0.6474],[218,-0.6445],[220,-0.6526],[222,-0.6451],[226,-0.6112],[228,-0.6032],[230,-0.602],[232,-0.5774],[234,-0.5549],[236,-0.5556],[238,-0.5632],[240,-0.5632],[242,-0.5765],[244,-0.6065],[246,-0.6065],[248,-0.6179],[250,-0.6249],[254,-0.6249],[256,-0.654],[258,-0.654],[260,-0.6608],[266,-0.6608],[268,-0.7238],[270,-0.7238],[272,-0.7608],[276,-0.7602],[278,-0.7579],[280,-0.7465],[282,-0.7435],[284,-0.7435],[286,-0.7654],[288,-0.7654],[290,-0.8087],[292,-0.8087],[294,-0.8325],[302,-0.8325],[304,-0.7864],[306,-0.7864],[308,-0.7357],[310,-0.7357],[312,-0.6898],[314,-0.6898],[316,-0.6512],[318,-0.6512],[320,-0.6493],[322,-0.6493],[324,-0.6545],[326,-0.6545],[328,-0.6661],[330,-0.6661],[332,-0.6384],[334,-0.636],[336,-0.3103],[338,-0.3336],[340,-0.4653],[342,-0.628],[344,-0.6869],[346,-0.7029],[348,-0.6301],[350,-0.6012],[352,-0.6552],[354,-0.7484],[356,-0.8899],[358,-1.0375],[360,-1.1738],[362,-1.3304],[364,-1.1796],[366,0.0666],[368,0.4703]]
       }
       chest_ctrl.Rotation.Z {
        type=MovieSceneScriptingFloatChannel
        default=2.4433
        thinned_from=140
        keys [[0,-2.7105],[94,-2.7083],[96,-2.6573],[98,-2.616],[100,-2.5369],[104,-2.4403],[106,-2.2797],[108,-2.1636],[110,-1.9718],[112,-1.8438],[114,-1.7576],[116,-1.7439],[118,-1.5485],[120,-1.1881],[122,-0.8452],[124,-0.7078],[126,-0.5995],[128,-0.7058],[130,-0.7575],[132,-0.6995],[134,-0.4085],[136,0.0469],[138,0.3132],[140,0.4679],[142,0.6195],[144,0.5679],[146,0.3552],[148,0.1109],[150,-0.1514],[152,-0.4318],[154,-0.6083],[156,-0.6749],[158,-0.7626],[160,-0.6588],[162,-0.6726],[164,-0.6337],[166,-0.6041],[168,-0.6223],[170,-0.6037],[172,-0.5838],[174,-0.5715],[176,-0.6033],[178,-0.6563],[180,-0.6868],[182,-0.7255],[184,-0.766],[186,-0.7807],[188,-0.7744],[190,-0.7418],[192,-0.6555],[194,-0.5241],[196,-0.3396],[198,-0.1567],[200,0.1569],[202,0.2966],[204,0.2966],[206,0.0304],[208,0.0305],[210,0.0735],[216,0.0735],[218,0.0614],[220,0.0083],[222,-0.0392],[224,-0.1441],[226,-0.2376],[228,-0.3202],[230,-0.4582],[232,-0.5606],[234,-0.6437],[236,-0.7135],[238,-0.739],[242,-0.7382],[244,-0.7079],[246,-0.7079],[248,-0.7037],[250,-0.7169],[254,-0.7169],[256,-0.7269],[258,-0.7269],[260,-0.771],[266,-0.771],[268,-0.7491],[270,-0.7491],[272,-0.7337],[274,-0.6771],[276,-0.6771],[278,-0.7506],[280,-0.8537],[282,-0.9163],[284,-0.9163],[286,-0.8649],[288,-0.8649],[290,-0.802],[292,-0.802],[294,-0.7534],[302,-0.7534],[304,-0.7886],[306,-0.7886],[308,-0.8444],[310,-0.8444],[312,-0.8793],[314,-0.8793],[316,-0.9411],[318,-0.9411],[320,-0.9848],[322,-0.9848],[324,-0.9598],[326,-0.9598],[328,-0.9523],[330,-0.9523],[332,-0.8409],[334,-0.7099],[336,-0.8009],[338,-1.1076],[340,-1.2617],[342,-2.0534],[344,-2.2751],[346,-2.5665],[350,-3.2268],[352,-3.2364],[356,-3.1611],[358,-2.9898],[360,-2.8694],[362,-2.5026],[364,-0.9902],[366,1.2981],[368,2.4433]]
       }
       chest_ctrl.Scale.X {type=MovieSceneScriptingFloatChannel, default=1, const [1,140]}
       chest_ctrl.Scale.Y {type=MovieSceneScriptingFloatChannel, default=1, const [1,140]}
       chest_ctrl.Scale.Z {type=MovieSceneScriptingFloatChannel, default=1, const [1,140]}
       chest_tan_ctrl.Location.X {
        type=MovieSceneScriptingFloatChannel
        default=-0.0086
        thinned_from=140
        keys [[0,0.0277],[92,0.0277],[94,0.0274],[98,0.0314],[100,0.0325],[102,0.0322],[104,0.0293],[106,0.0238],[108,0.0205],[112,0.0099],[114,0.0039],[116,0.0013],[118,-0.001],[120,-0.0053],[122,-0.0131],[124,-0.0161],[126,-0.0181],[128,-0.0176],[130,-0.0182],[134,-0.021],[136,-0.0201],[138,-0.0159],[140,-0.0124],[142,-0.0131],[144,-0.0142],[146,-0.0116],[148,-0.0086],[150,-0.0066],[152,-0.0049],[154,-0.0026],[156,-0.0014],[158,-0.0006],[160,0.0039],[162,0.0064],[166,0.012],[168,0.0134],[170,0.0152],[172,0.0189],[174,0.0211],[176,0.0205],[178,0.0193],[180,0.0197],[182,0.0208],[184,0.0217],[186,0.0219],[188,0.0225],[190,0.0235],[192,0.0248],[194,0.0258],[198,0.0289],[200,0.0307],[202,0.0321],[204,0.0321],[206,0.0304],[208,0.0304],[210,0.0296],[216,0.0296],[218,0.0294],[220,0.0296],[222,0.0292],[226,0.0272],[230,0.0262],[234,0.0237],[240,0.0235],[242,0.0257],[244,0.0251],[246,0.0251],[248,0.0272],[250,0.0275],[254,0.0275],[256,0.028],[258,0.028],[260,0.0282],[266,0.0282],[268,0.0311],[270,0.0311],[272,0.0329],[276,0.033],[278,0.0327],[280,0.0319],[282,0.0317],[284,0.0317],[286,0.0328],[288,0.0328],[290,0.0349],[292,0.0349],[294,0.0361],[302,0.0361],[304,0.0339],[306,0.0339],[308,0.0315],[310,0.0315],[312,0.0294],[314,0.0294],[316,0.0275],[322,0.0273],[324,0.0276],[326,0.0276],[328,0.0282],[330,0.0282],[332,0.0273],[334,0.0275],[336,0.0132],[338,0.0153],[340,0.0229],[342,0.0326],[344,0.0373],[346,0.0384],[348,0.0401],[350,0.0391],[352,0.0441],[354,0.0472],[356,0.0475],[358,0.0548],[360,0.0557],[362,0.0603],[364,0.0511],[366,0.0011],[368,-0.0086]]
       }
       chest_tan_ctrl.Location.Y {
        type=MovieSceneScriptingFloatChannel
        default=-1.0843
        thinned_from=140
        keys [[0,-2.0142],[94,-2.0149],[102,-2.006],[104,-1.9944],[106,-1.9754],[110,-1.9261],[112,-1.8911],[122,-1.675],[124,-1.6437],[126,-1.6168],[128,-1.5833],[134,-1.5003],[142,-1.3815],[146,-1.3197],[148,-1.2936],[152,-1.25],[154,-1.2304],[160,-1.1833],[164,-1.1648],[172,-1.1356],[186,-1.0969],[192,-1.0845],[198,-1.0769],[246,-1.0767],[306,-1.0711],[308,-1.0677],[310,-1.0677],[312,-1.0655],[332,-1.0621],[334,-1.0641],[336,-1.0564],[338,-1.0092],[340,-0.9444],[342,-0.8648],[344,-0.7927],[346,-0.7268],[348,-0.6666],[350,-0.6309],[352,-0.6011],[354,-0.5949],[356,-0.6346],[358,-0.7239],[360,-0.8536],[362,-0.9719],[364,-1.0261],[366,-1.1047],[368,-1.0843]]
       }
       chest_tan_ctrl.Location.Z {
        type=MovieSceneScriptingFloatChannel
        default=0.2307
        thinned_from=140
        keys [[0,0.7639],[94,0.7635],[100,0.7618],[102,0.7553],[104,0.7374],[106,0.71],[108,0.6765],[110,0.6395],[114,0.5563],[116,0.516],[118,0.4794],[120,0.4457],[122,0.4177],[124,0.397],[126,0.3804],[128,0.3626],[130,0.3467],[132,0.3321],[136,0.3051],[138,0.2927],[140,0.2812],[142,0.2708],[144,0.261],[146,0.2524],[148,0.245],[150,0.2387],[152,0.2333],[154,0.2287],[156,0.2248],[158,0.2216],[160,0.2189],[162,0.2169],[176,0.2093],[190,0.2051],[334,0.2039],[336,0.2597],[338,0.4107],[340,0.6335],[342,0.9044],[344,1.1995],[346,1.7043],[348,1.7678],[350,2.2007],[352,2.1443],[354,2.2003],[356,2.2005],[358,1.4955],[360,1.1138],[362,0.6218],[364,0.4794],[366,0.227],[368,0.2307]]
       }
       chest_tan_ctrl.Rotation.X {
        type=MovieSceneScriptingFloatChannel
        default=7.8037
        thinned_from=140
        keys [[0,11.0445],[94,11.0445],[102,10.9747],[104,10.8895],[110,10.448],[112,10.2365],[122,9.0494],[124,8.8989],[126,8.7746],[128,8.6016],[130,8.4632],[134,8.2438],[142,7.7235],[144,7.5731],[146,7.4382],[148,7.328],[154,7.081],[160,6.8955],[172,6.7023],[192,6.4967],[198,6.4589],[204,6.4601],[218,6.4732],[222,6.4838],[254,6.4535],[256,6.4336],[266,6.4401],[268,6.4295],[270,6.435],[272,6.4213],[282,6.4191],[284,6.4269],[286,6.4092],[288,6.4177],[290,6.4086],[292,6.4178],[306,6.4252],[308,6.4008],[310,6.4008],[312,6.3854],[332,6.3621],[334,6.3774],[336,6.3209],[338,5.9664],[340,5.4569],[344,4.1308],[346,3.3927],[348,2.8153],[350,2.2601],[352,1.9925],[354,1.8973],[356,2.3038],[358,3.4521],[360,4.7005],[362,5.8127],[364,6.6646],[366,7.7741],[368,7.8037]]
       }
       chest_tan_ctrl.Rotation.Y {
        type=MovieSceneScriptingFloatChannel
        default=0
        thinned_from=140
        keys [[0,0],[158,0],[160,0.0208],[162,0.0395],[166,0.082],[168,0.0925],[170,0.1062],[172,0.134],[174,0.1507],[176,0.1461],[178,0.1378],[180,0.1409],[182,0.1492],[184,0.1554],[186,0.1575],[188,0.162],[190,0.1697],[192,0.1798],[194,0.1878],[196,0.2006],[198,0.2121],[200,0.2259],[202,0.2367],[204,0.2367],[206,0.2233],[208,0.2233],[210,0.2173],[212,0.2151],[214,0.2086],[216,0.2151],[218,0.2173],[220,0.2127],[222,0.2142],[224,0.2062],[226,0.1987],[228,0.1944],[230,0.1911],[232,0.1808],[234,0.1716],[236,0.1704],[238,0.1739],[240,0.1704],[242,0.1837],[244,0.1889],[246,0.1837],[248,0.1957],[254,0.1957],[256,0.2033],[258,0.2033],[260,0.2046],[266,0.2046],[268,0.2264],[270,0.2264],[272,0.2391],[274,0.2399],[276,0.2431],[280,0.2318],[282,0.2296],[284,0.2296],[286,0.2382],[288,0.2382],[290,0.2539],[292,0.2539],[294,0.2628],[296,0.2615],[298,0.2628],[302,0.2628],[304,0.2468],[306,0.2468],[308,0.2291],[310,0.2291],[312,0.2132],[314,0.2132],[316,0.1992],[318,0.1992],[320,0.1978],[322,0.1978],[324,0.2017],[330,0.2017],[332,0.1981],[334,0.2003],[336,0.0943],[338,0.1098],[340,0.1709],[342,0.2575],[344,0.3083],[346,0.3573],[348,0.3674],[350,0.4107],[352,0.4299],[354,0.4619],[356,0.4962],[358,0.4841],[360,0.4768],[362,0.4714],[364,0.3844],[366,0.0249],[368,0]]
       }
       chest_tan_ctrl.Rotation.Z {
        type=MovieSceneScriptingFloatChannel
        default=0
        thinned_from=140
        keys [[0,0.9913],[94,0.9913],[96,0.9748],[98,0.9651],[100,0.9403],[102,0.9216],[104,0.8944],[106,0.8225],[108,0.7729],[110,0.6965],[112,0.6393],[114,0.5965],[116,0.5858],[118,0.5132],[120,0.3815],[122,0.2539],[124,0.205],[126,0.1696],[128,0.2111],[130,0.2307],[132,0.2135],[134,0.1161],[136,0],[148,0],[150,0.0512],[152,0.1437],[154,0.2018],[156,0.2238],[158,0.2523],[160,0.2182],[162,0.2207],[164,0.206],[166,0.195],[168,0.2003],[170,0.1934],[172,0.1861],[174,0.1818],[176,0.1921],[178,0.2098],[180,0.2203],[182,0.2331],[184,0.2466],[186,0.2514],[188,0.2494],[190,0.2385],[192,0.2093],[194,0.1646],[196,0.1023],[198,0.0407],[200,0],[222,0],[224,0.0281],[226,0.0594],[228,0.0865],[230,0.132],[232,0.1664],[234,0.1943],[236,0.2174],[238,0.229],[240,0.2174],[242,0.2188],[244,0.2107],[246,0.2188],[248,0.2163],[254,0.2163],[256,0.2181],[258,0.2181],[260,0.2321],[266,0.2321],[268,0.2237],[270,0.2237],[272,0.2197],[274,0.2015],[276,0.2121],[278,0.2259],[280,0.2602],[282,0.2811],[284,0.2811],[286,0.2626],[288,0.2626],[290,0.2408],[292,0.2408],[294,0.2253],[296,0.2215],[298,0.2253],[302,0.2253],[304,0.2353],[306,0.2353],[308,0.2504],[310,0.2504],[312,0.2587],[314,0.2587],[316,0.278],[318,0.278],[320,0.291],[322,0.291],[324,0.2782],[330,0.2782],[332,0.2466],[334,0.2055],[336,0.2411],[338,0.3437],[340,0.3869],[342,0.6357],[344,0.7003],[346,0.9292],[348,0.8924],[350,1.1833],[352,0.9892],[354,0.9688],[356,1.1529],[358,0.8997],[360,1.0713],[362,0.9502],[364,0.3461],[366,0],[368,0]]
       }
       chest_tan_ctrl.Scale.X {type=MovieSceneScriptingFloatChannel, default=1, const [1,140]}
       chest_tan_ctrl.Scale.Y {type=MovieSceneScriptingFloatChannel, default=1, const [1,140]}
       chest_tan_ctrl.Scale.Z {type=MovieSceneScriptingFloatChannel, default=1, const [1,140]}
      ]
     }
     {
      type=MovieSceneControlRigParameterSection
      channels [
       global_ctrl.Location.X {type=MovieSceneScriptingFloatChannel, default=0, const [0,6]}
       global_ctrl.Location.Y {type=MovieSceneScriptingFloatChannel, default=0, const [0,6]}
       global_ctrl.Location.Z {type=MovieSceneScriptingFloatChannel, default=0, const [0,6]}
       global_ctrl.Rotation.X {type=MovieSceneScriptingFloatChannel, default=0, const [0,6]}
       global_ctrl.Rotation.Y {type=MovieSceneScriptingFloatChannel, default=0, const [0,6]}
       global_ctrl.Rotation.Z {type=MovieSceneScriptingFloatChannel, default=0, const [0,6]}
       root_ctrl.Location.X {type=MovieSceneScriptingFloatChannel, default=0, const [0,6]}
       root_ctrl.Location.Y {type=MovieSceneScriptingFloatChannel, default=0, const [0,6]}
       root_ctrl.Location.Z {type=MovieSceneScriptingFloatChannel, default=0, const [0,6]}
       root_ctrl.Rotation.X {type=MovieSceneScriptingFloatChannel, default=0, const [0,6]}
       root_ctrl.Rotation.Y {type=MovieSceneScriptingFloatChannel, default=0, const [0,6]}
       root_ctrl.Rotation.Z {type=MovieSceneScriptingFloatChannel, default=0, const [0,6]}
       root_ctrl.Scale.X {type=MovieSceneScriptingFloatChannel, default=0, const [0,6]}
       root_ctrl.Scale.Y {type=MovieSceneScriptingFloatChannel, default=0, const [0,6]}
       root_ctrl.Scale.Z {type=MovieSceneScriptingFloatChannel, default=0, const [0,6]}
       body_offset_ctrl.Location.X {type=MovieSceneScriptingFloatChannel, default=0, const [0,6]}
       body_offset_ctrl.Location.Y {type=MovieSceneScriptingFloatChannel, default=0, const [0,6]}
       body_offset_ctrl.Location.Z {type=MovieSceneScriptingFloatChannel, default=0, const [0,6]}
       body_offset_ctrl.Rotation.X {type=MovieSceneScriptingFloatChannel, default=0, const [0,6]}
       body_offset_ctrl.Rotation.Y {type=MovieSceneScriptingFloatChannel, default=0, const [0,6]}
       body_offset_ctrl.Rotation.Z {type=MovieSceneScriptingFloatChannel, default=0, const [0,6]}
       body_offset_ctrl.Scale.X {type=MovieSceneScriptingFloatChannel, default=0, const [0,6]}
       body_offset_ctrl.Scale.Y {type=MovieSceneScriptingFloatChannel, default=0, const [0,6]}
       body_offset_ctrl.Scale.Z {type=MovieSceneScriptingFloatChannel, default=0, const [0,6]}
       thigh_l_fk_ctrl.Location.X {type=MovieSceneScriptingFloatChannel, default=0, const [0,6]}
       thigh_l_fk_ctrl.Location.Y {type=MovieSceneScriptingFloatChannel, default=0, const [0,6]}
       thigh_l_fk_ctrl.Location.Z {type=MovieSceneScriptingFloatChannel, default=0, const [0,6]}
       thigh_l_fk_ctrl.Rotation.X {type=MovieSceneScriptingFloatChannel, default=0, const [0,6]}
       thigh_l_fk_ctrl.Rotation.Y {type=MovieSceneScriptingFloatChannel, default=0, const [0,6]}
       thigh_l_fk_ctrl.Rotation.Z {type=MovieSceneScriptingFloatChannel, default=0, const [0,6]}
       thigh_l_fk_ctrl.Scale.X {type=MovieSceneScriptingFloatChannel, default=0, const [0,6]}
       thigh_l_fk_ctrl.Scale.Y {type=MovieSceneScriptingFloatChannel, default=0, const [0,6]}
       thigh_l_fk_ctrl.Scale.Z {type=MovieSceneScriptingFloatChannel, default=0, const [0,6]}
       calf_l_fk_ctrl.Location.X {type=MovieSceneScriptingFloatChannel, default=0, const [0,6]}
       calf_l_fk_ctrl.Location.Y {type=MovieSceneScriptingFloatChannel, default=0, const [0,6]}
       calf_l_fk_ctrl.Location.Z {type=MovieSceneScriptingFloatChannel, default=0, const [0,6]}
       calf_l_fk_ctrl.Rotation.X {type=MovieSceneScriptingFloatChannel, default=0, const [0,6]}
       calf_l_fk_ctrl.Rotation.Y {type=MovieSceneScriptingFloatChannel, default=0, const [0,6]}
       calf_l_fk_ctrl.Rotation.Z {type=MovieSceneScriptingFloatChannel, default=0, const [0,6]}
       calf_l_fk_ctrl.Scale.X {type=MovieSceneScriptingFloatChannel, default=0, const [0,6]}
       calf_l_fk_ctrl.Scale.Y {type=MovieSceneScriptingFloatChannel, default=0, const [0,6]}
       calf_l_fk_ctrl.Scale.Z {type=MovieSceneScriptingFloatChannel, default=0, const [0,6]}
       foot_l_fk_ctrl.Location.X {type=MovieSceneScriptingFloatChannel, default=0, const [0,6]}
       foot_l_fk_ctrl.Location.Y {type=MovieSceneScriptingFloatChannel, default=0, const [0,6]}
       foot_l_fk_ctrl.Location.Z {type=MovieSceneScriptingFloatChannel, default=0, const [0,6]}
       foot_l_fk_ctrl.Rotation.X {type=MovieSceneScriptingFloatChannel, default=0, const [0,6]}
       foot_l_fk_ctrl.Rotation.Y {type=MovieSceneScriptingFloatChannel, default=0, const [0,6]}
       foot_l_fk_ctrl.Rotation.Z {type=MovieSceneScriptingFloatChannel, default=0, const [0,6]}
       foot_l_fk_ctrl.Scale.X {type=MovieSceneScriptingFloatChannel, default=0, const [0,6]}
       foot_l_fk_ctrl.Scale.Y {type=MovieSceneScriptingFloatChannel, default=0, const [0,6]}
       foot_l_fk_ctrl.Scale.Z {type=MovieSceneScriptingFloatChannel, default=0, const [0,6]}
       ball_l_fk_ctrl.Location.X {type=MovieSceneScriptingFloatChannel, default=0, const [0,6]}
       ball_l_fk_ctrl.Location.Y {type=MovieSceneScriptingFloatChannel, default=0, const [0,6]}
       ball_l_fk_ctrl.Location.Z {type=MovieSceneScriptingFloatChannel, default=0, const [0,6]}
       ball_l_fk_ctrl.Rotation.X {type=MovieSceneScriptingFloatChannel, default=0, const [0,6]}
       ball_l_fk_ctrl.Rotation.Y {type=MovieSceneScriptingFloatChannel, default=0, const [0,6]}
       ball_l_fk_ctrl.Rotation.Z {type=MovieSceneScriptingFloatChannel, default=0, const [0,6]}
       ball_l_fk_ctrl.Scale.X {type=MovieSceneScriptingFloatChannel, default=0, const [0,6]}
       ball_l_fk_ctrl.Scale.Y {type=MovieSceneScriptingFloatChannel, default=0, const [0,6]}
       ball_l_fk_ctrl.Scale.Z {type=MovieSceneScriptingFloatChannel, default=0, const [0,6]}
       foot_l_ik_ctrl.Location.X {type=MovieSceneScriptingFloatChannel, default=0, keys [[324,0],[335,-0.1395],[349,-0.014],[351,-0.0139],[357,-0.0321],[362,0]]}
       foot_l_ik_ctrl.Location.Y {type=MovieSceneScriptingFloatChannel, default=0, keys [[324,0],[335,-0.0697],[349,-0.0569],[351,-0.0569],[357,-0.1316],[362,0]]}
       foot_l_ik_ctrl.Location.Z {type=MovieSceneScriptingFloatChannel, default=0, keys [[324,0],[335,-0.044],[349,-0.0171],[351,-0.0171],[357,-0.0395],[362,0]]}
       foot_l_ik_ctrl.Rotation.X {type=MovieSceneScriptingFloatChannel, default=0, keys [[324,0],[335,-0.0699],[349,0.0884],[351,0.089],[357,0.1732],[362,0]]}
       foot_l_ik_ctrl.Rotation.Y {type=MovieSceneScriptingFloatChannel, default=0, keys [[324,0],[335,-0.0952],[349,-0.0169],[351,-0.0168],[357,-0.0389],[362,0]]}
       foot_l_ik_ctrl.Rotation.Z {type=MovieSceneScriptingFloatChannel, default=0, keys [[324,0],[335,-0.0042],[349,-0.0034],[351,-0.0034],[357,-0.0079],[362,0]]}
       foot_l_ik_ctrl.Scale.X {type=MovieSceneScriptingFloatChannel, default=0, const [0,6]}
       foot_l_ik_ctrl.Scale.Y {type=MovieSceneScriptingFloatChannel, default=0, const [0,6]}
       foot_l_ik_ctrl.Scale.Z {type=MovieSceneScriptingFloatChannel, default=0, const [0,6]}
       ball_l_ik_ctrl.Location.X {type=MovieSceneScriptingFloatChannel, default=0, const [0,6]}
       ball_l_ik_ctrl.Location.Y {type=MovieSceneScriptingFloatChannel, default=0, const [0,6]}
       ball_l_ik_ctrl.Location.Z {type=MovieSceneScriptingFloatChannel, default=0, const [0,6]}
       ball_l_ik_ctrl.Rotation.X {type=MovieSceneScriptingFloatChannel, default=0, const [0,6]}
       ball_l_ik_ctrl.Rotation.Y {type=MovieSceneScriptingFloatChannel, default=0, const [0,6]}
       ball_l_ik_ctrl.Rotation.Z {type=MovieSceneScriptingFloatChannel, default=0, const [0,6]}
       ball_l_ik_ctrl.Scale.X {type=MovieSceneScriptingFloatChannel, default=0, const [0,6]}
       ball_l_ik_ctrl.Scale.Y {type=MovieSceneScriptingFloatChannel, default=0, const [0,6]}
       ball_l_ik_ctrl.Scale.Z {type=MovieSceneScriptingFloatChannel, default=0, const [0,6]}
       foot_bk1_l_ctrl.Location.X {type=MovieSceneScriptingFloatChannel, default=0, const [0,6]}
       foot_bk1_l_ctrl.Location.Y {type=MovieSceneScriptingFloatChannel, default=0, const [0,6]}
       foot_bk1_l_ctrl.Location.Z {type=MovieSceneScriptingFloatChannel, default=0, const [0,6]}
       foot_bk1_l_ctrl.Rotation.X {type=MovieSceneScriptingFloatChannel, default=0, const [0,6]}
       foot_bk1_l_ctrl.Rotation.Y {type=MovieSceneScriptingFloatChannel, default=0, const [0,6]}
       foot_bk1_l_ctrl.Rotation.Z {type=MovieSceneScriptingFloatChannel, default=0, const [0,6]}
       foot_bk1_l_ctrl.Scale.X {type=MovieSceneScriptingFloatChannel, default=0, const [0,6]}
       foot_bk1_l_ctrl.Scale.Y {type=MovieSceneScriptingFloatChannel, default=0, const [0,6]}
       foot_bk1_l_ctrl.Scale.Z {type=MovieSceneScriptingFloatChannel, default=0, const [0,6]}
       foot_roll_l_ctrl.Location.X {type=MovieSceneScriptingFloatChannel, default=0, const [0,6]}
       foot_roll_l_ctrl.Location.Y {type=MovieSceneScriptingFloatChannel, default=0, const [0,6]}
       foot_roll_l_ctrl.Location.Z {type=MovieSceneScriptingFloatChannel, default=0, const [0,6]}
       foot_roll_l_ctrl.Rotation.X {type=MovieSceneScriptingFloatChannel, default=0, const [0,6]}
       foot_roll_l_ctrl.Rotation.Y {type=MovieSceneScriptingFloatChannel, default=0, const [0,6]}
       foot_roll_l_ctrl.Rotation.Z {type=MovieSceneScriptingFloatChannel, default=0, const [0,6]}
       foot_roll_l_ctrl.Scale.X {type=MovieSceneScriptingFloatChannel, default=0, const [0,6]}
       foot_roll_l_ctrl.Scale.Y {type=MovieSceneScriptingFloatChannel, default=0, const [0,6]}
       foot_roll_l_ctrl.Scale.Z {type=MovieSceneScriptingFloatChannel, default=0, const [0,6]}
       roll_blend_l_ctrl {type=MovieSceneScriptingFloatChannel, default=0, const [0,6]}
       tip_l_ctrl.Location.X {type=MovieSceneScriptingFloatChannel, default=0, const [0,6]}
       tip_l_ctrl.Location.Y {type=MovieSceneScriptingFloatChannel, default=0, const [0,6]}
       tip_l_ctrl.Location.Z {type=MovieSceneScriptingFloatChannel, default=0, const [0,6]}
       tip_l_ctrl.Rotation.X {type=MovieSceneScriptingFloatChannel, default=0, const [0,6]}
       tip_l_ctrl.Rotation.Y {type=MovieSceneScriptingFloatChannel, default=0, const [0,6]}
       tip_l_ctrl.Rotation.Z {type=MovieSceneScriptingFloatChannel, default=0, const [0,6]}
       tip_l_ctrl.Scale.X {type=MovieSceneScriptingFloatChannel, default=0, const [0,6]}
       tip_l_ctrl.Scale.Y {type=MovieSceneScriptingFloatChannel, default=0, const [0,6]}
       tip_l_ctrl.Scale.Z {type=MovieSceneScriptingFloatChannel, default=0, const [0,6]}
       heel_l_ctrl.Location.X {type=MovieSceneScriptingFloatChannel, default=0, const [0,6]}
       heel_l_ctrl.Location.Y {type=MovieSceneScriptingFloatChannel, default=0, const [0,6]}
       heel_l_ctrl.Location.Z {type=MovieSceneScriptingFloatChannel, default=0, const [0,6]}
       heel_l_ctrl.Rotation.X {type=MovieSceneScriptingFloatChannel, default=0, const [0,6]}
       heel_l_ctrl.Rotation.Y {type=MovieSceneScriptingFloatChannel, default=0, const [0,6]}
       heel_l_ctrl.Rotation.Z {type=MovieSceneScriptingFloatChannel, default=0, const [0,6]}
       heel_l_ctrl.Scale.X {type=MovieSceneScriptingFloatChannel, default=0, const [0,6]}
       heel_l_ctrl.Scale.Y {type=MovieSceneScriptingFloatChannel, default=0, const [0,6]}
       heel_l_ctrl.Scale.Z {type=MovieSceneScriptingFloatChannel, default=0, const [0,6]}
       leg_l_softness {type=MovieSceneScriptingFloatChannel, default=0, const [0,6]}
       leg_l_pv_ik_ctrl.Location.X {type=MovieSceneScriptingFloatChannel, default=0, keys [[324,0],[335,-0.0351],[349,0.0181],[351,0.0183],[357,0.0319],[362,0]]}
       leg_l_pv_ik_ctrl.Location.Y {type=MovieSceneScriptingFloatChannel, default=0, keys [[324,0],[335,-0.0287],[349,0.0037],[351,0.0038],[357,0.006],[362,0]]}
       leg_l_pv_ik_ctrl.Location.Z {type=MovieSceneScriptingFloatChannel, default=0, keys [[324,0],[335,-0.0632],[349,0.0562],[351,0.0566],[357,0.1047],[362,0]]}
       leg_l_pv_ik_ctrl.Rotation.X {type=MovieSceneScriptingFloatChannel, default=0, keys [[324,0],[335,0.0286],[349,-0.0392],[351,-0.0394],[357,-0.0778],[362,0]]}
       leg_l_pv_ik_ctrl.Rotation.Y {type=MovieSceneScriptingFloatChannel, default=0, keys [[324,0],[335,-0.0963],[349,-0.0064],[351,-0.0063],[357,-0.0147],[362,0]]}
       leg_l_pv_ik_ctrl.Rotation.Z {type=MovieSceneScriptingFloatChannel, default=0, keys [[324,0],[335,-0.0326],[349,-0.0036],[351,-0.0036],[357,-0.0084],[362,0]]}
       leg_l_pv_ik_ctrl.Scale.X {type=MovieSceneScriptingFloatChannel, default=0, const [0,6]}
       leg_l_pv_ik_ctrl.Scale.Y {type=MovieSceneScriptingFloatChannel, default=0, const [0,6]}
       leg_l_pv_ik_ctrl.Scale.Z {type=MovieSceneScriptingFloatChannel, default=0, const [0,6]}
       thigh_r_fk_ctrl.Location.X {type=MovieSceneScriptingFloatChannel, default=0, const [0,6]}
       thigh_r_fk_ctrl.Location.Y {type=MovieSceneScriptingFloatChannel, default=0, const [0,6]}
       thigh_r_fk_ctrl.Location.Z {type=MovieSceneScriptingFloatChannel, default=0, const [0,6]}
       thigh_r_fk_ctrl.Rotation.X {type=MovieSceneScriptingFloatChannel, default=0, const [0,6]}
       thigh_r_fk_ctrl.Rotation.Y {type=MovieSceneScriptingFloatChannel, default=0, const [0,6]}
       thigh_r_fk_ctrl.Rotation.Z {type=MovieSceneScriptingFloatChannel, default=0, const [0,6]}
       thigh_r_fk_ctrl.Scale.X {type=MovieSceneScriptingFloatChannel, default=0, const [0,6]}
       thigh_r_fk_ctrl.Scale.Y {type=MovieSceneScriptingFloatChannel, default=0, const [0,6]}
       thigh_r_fk_ctrl.Scale.Z {type=MovieSceneScriptingFloatChannel, default=0, const [0,6]}
       calf_r_fk_ctrl.Location.X {type=MovieSceneScriptingFloatChannel, default=0, const [0,6]}
       calf_r_fk_ctrl.Location.Y {type=MovieSceneScriptingFloatChannel, default=0, const [0,6]}
       calf_r_fk_ctrl.Location.Z {type=MovieSceneScriptingFloatChannel, default=0, const [0,6]}
       calf_r_fk_ctrl.Rotation.X {type=MovieSceneScriptingFloatChannel, default=0, const [0,6]}
       calf_r_fk_ctrl.Rotation.Y {type=MovieSceneScriptingFloatChannel, default=0, const [0,6]}
       calf_r_fk_ctrl.Rotation.Z {type=MovieSceneScriptingFloatChannel, default=0, const [0,6]}
       calf_r_fk_ctrl.Scale.X {type=MovieSceneScriptingFloatChannel, default=0, const [0,6]}
       calf_r_fk_ctrl.Scale.Y {type=MovieSceneScriptingFloatChannel, default=0, const [0,6]}
       calf_r_fk_ctrl.Scale.Z {type=MovieSceneScriptingFloatChannel, default=0, const [0,6]}
       foot_r_fk_ctrl.Location.X {type=MovieSceneScriptingFloatChannel, default=0, const [0,6]}
       foot_r_fk_ctrl.Location.Y {type=MovieSceneScriptingFloatChannel, default=0, const [0,6]}
       foot_r_fk_ctrl.Location.Z {type=MovieSceneScriptingFloatChannel, default=0, const [0,6]}
       foot_r_fk_ctrl.Rotation.X {type=MovieSceneScriptingFloatChannel, default=0, const [0,6]}
       foot_r_fk_ctrl.Rotation.Y {type=MovieSceneScriptingFloatChannel, default=0, const [0,6]}
       foot_r_fk_ctrl.Rotation.Z {type=MovieSceneScriptingFloatChannel, default=0, const [0,6]}
       foot_r_fk_ctrl.Scale.X {type=MovieSceneScriptingFloatChannel, default=0, const [0,6]}
       foot_r_fk_ctrl.Scale.Y {type=MovieSceneScriptingFloatChannel, default=0, const [0,6]}
       foot_r_fk_ctrl.Scale.Z {type=MovieSceneScriptingFloatChannel, default=0, const [0,6]}
       ball_r_fk_ctrl.Location.X {type=MovieSceneScriptingFloatChannel, default=0, thinned_from=6, keys [[324,0],[362,0]]}
       ball_r_fk_ctrl.Location.Y {type=MovieSceneScriptingFloatChannel, default=0, thinned_from=6, keys [[324,0],[362,0]]}
       ball_r_fk_ctrl.Location.Z {type=MovieSceneScriptingFloatChannel, default=0, thinned_from=6, keys [[324,0],[362,0]]}
       ball_r_fk_ctrl.Rotation.X {type=MovieSceneScriptingFloatChannel, default=0, const [0,6]}
       ball_r_fk_ctrl.Rotation.Y {type=MovieSceneScriptingFloatChannel, default=0, const [0,6]}
       ball_r_fk_ctrl.Rotation.Z {type=MovieSceneScriptingFloatChannel, default=0, const [0,6]}
       ball_r_fk_ctrl.Scale.X {type=MovieSceneScriptingFloatChannel, default=0, const [0,6]}
       ball_r_fk_ctrl.Scale.Y {type=MovieSceneScriptingFloatChannel, default=0, const [0,6]}
       ball_r_fk_ctrl.Scale.Z {type=MovieSceneScriptingFloatChannel, default=0, const [0,6]}
       foot_r_ik_ctrl.Location.X {type=MovieSceneScriptingFloatChannel, default=0, keys [[324,0],[335,0.071],[349,0.0007],[351,0.0006],[357,0.0015],[362,0]]}
       foot_r_ik_ctrl.Location.Y {type=MovieSceneScriptingFloatChannel, default=0, keys [[324,0],[335,-0.0488],[349,-0.0245],[351,-0.0245],[357,-0.0566],[362,0]]}
       foot_r_ik_ctrl.Location.Z {type=MovieSceneScriptingFloatChannel, default=0, keys [[324,0],[335,0.0714],[349,0.0191],[351,0.019],[357,0.0441],[362,0]]}
       foot_r_ik_ctrl.Rotation.X {type=MovieSceneScriptingFloatChannel, default=0, keys [[324,0],[335,-0.001],[349,-0.0623],[351,-0.0625],[357,-0.1447],[362,0]]}
       foot_r_ik_ctrl.Rotation.Y {type=MovieSceneScriptingFloatChannel, default=0, keys [[324,0],[335,0.0271],[349,-0.0045],[351,-0.0045],[357,-0.0071],[362,0]]}
       foot_r_ik_ctrl.Rotation.Z {type=MovieSceneScriptingFloatChannel, default=0, keys [[324,0],[335,-0.0129],[349,-0.0178],[351,-0.0178],[357,-0.0412],[362,0]]}
       foot_r_ik_ctrl.Scale.X {type=MovieSceneScriptingFloatChannel, default=0, const [0,6]}
       foot_r_ik_ctrl.Scale.Y {type=MovieSceneScriptingFloatChannel, default=0, const [0,6]}
       foot_r_ik_ctrl.Scale.Z {type=MovieSceneScriptingFloatChannel, default=0, const [0,6]}
       ball_r_ik_ctrl.Location.X {type=MovieSceneScriptingFloatChannel, default=0, thinned_from=6, keys [[324,0],[362,0]]}
       ball_r_ik_ctrl.Location.Y {type=MovieSceneScriptingFloatChannel, default=0, thinned_from=6, keys [[324,0],[362,0]]}
       ball_r_ik_ctrl.Location.Z {type=MovieSceneScriptingFloatChannel, default=0, thinned_from=6, keys [[324,0],[362,0]]}
       ball_r_ik_ctrl.Rotation.X {type=MovieSceneScriptingFloatChannel, default=0, const [0,6]}
       ball_r_ik_ctrl.Rotation.Y {type=MovieSceneScriptingFloatChannel, default=0, const [0,6]}
       ball_r_ik_ctrl.Rotation.Z {type=MovieSceneScriptingFloatChannel, default=0, const [0,6]}
       ball_r_ik_ctrl.Scale.X {type=MovieSceneScriptingFloatChannel, default=0, const [0,6]}
       ball_r_ik_ctrl.Scale.Y {type=MovieSceneScriptingFloatChannel, default=0, const [0,6]}
       ball_r_ik_ctrl.Scale.Z {type=MovieSceneScriptingFloatChannel, default=0, const [0,6]}
       foot_bk1_r_ctrl.Location.X {type=MovieSceneScriptingFloatChannel, default=0, const [0,6]}
       foot_bk1_r_ctrl.Location.Y {type=MovieSceneScriptingFloatChannel, default=0, const [0,6]}
       foot_bk1_r_ctrl.Location.Z {type=MovieSceneScriptingFloatChannel, default=0, const [0,6]}
       foot_bk1_r_ctrl.Rotation.X {type=MovieSceneScriptingFloatChannel, default=0, const [0,6]}
       foot_bk1_r_ctrl.Rotation.Y {type=MovieSceneScriptingFloatChannel, default=0, const [0,6]}
       foot_bk1_r_ctrl.Rotation.Z {type=MovieSceneScriptingFloatChannel, default=0, const [0,6]}
       foot_bk1_r_ctrl.Scale.X {type=MovieSceneScriptingFloatChannel, default=0, const [0,6]}
       foot_bk1_r_ctrl.Scale.Y {type=MovieSceneScriptingFloatChannel, default=0, const [0,6]}
       foot_bk1_r_ctrl.Scale.Z {type=MovieSceneScriptingFloatChannel, default=0, const [0,6]}
       foot_roll_r_ctrl.Location.X {type=MovieSceneScriptingFloatChannel, default=0, const [0,6]}
       foot_roll_r_ctrl.Location.Y {type=MovieSceneScriptingFloatChannel, default=0, const [0,6]}
       foot_roll_r_ctrl.Location.Z {type=MovieSceneScriptingFloatChannel, default=0, const [0,6]}
       foot_roll_r_ctrl.Rotation.X {type=MovieSceneScriptingFloatChannel, default=0, const [0,6]}
       foot_roll_r_ctrl.Rotation.Y {type=MovieSceneScriptingFloatChannel, default=0, const [0,6]}
       foot_roll_r_ctrl.Rotation.Z {type=MovieSceneScriptingFloatChannel, default=0, const [0,6]}
       foot_roll_r_ctrl.Scale.X {type=MovieSceneScriptingFloatChannel, default=0, const [0,6]}
       foot_roll_r_ctrl.Scale.Y {type=MovieSceneScriptingFloatChannel, default=0, const [0,6]}
       foot_roll_r_ctrl.Scale.Z {type=MovieSceneScriptingFloatChannel, default=0, const [0,6]}
       roll_blend_r_ctrl {type=MovieSceneScriptingFloatChannel, default=0, const [0,6]}
       tip_r_ctrl.Location.X {type=MovieSceneScriptingFloatChannel, default=0, const [0,6]}
       tip_r_ctrl.Location.Y {type=MovieSceneScriptingFloatChannel, default=0, const [0,6]}
       tip_r_ctrl.Location.Z {type=MovieSceneScriptingFloatChannel, default=0, const [0,6]}
       tip_r_ctrl.Rotation.X {type=MovieSceneScriptingFloatChannel, default=0, const [0,6]}
       tip_r_ctrl.Rotation.Y {type=MovieSceneScriptingFloatChannel, default=0, const [0,6]}
       tip_r_ctrl.Rotation.Z {type=MovieSceneScriptingFloatChannel, default=0, const [0,6]}
       tip_r_ctrl.Scale.X {type=MovieSceneScriptingFloatChannel, default=0, const [0,6]}
       tip_r_ctrl.Scale.Y {type=MovieSceneScriptingFloatChannel, default=0, const [0,6]}
       tip_r_ctrl.Scale.Z {type=MovieSceneScriptingFloatChannel, default=0, const [0,6]}
       heel_r_ctrl.Location.X {type=MovieSceneScriptingFloatChannel, default=0, const [0,6]}
       heel_r_ctrl.Location.Y {type=MovieSceneScriptingFloatChannel, default=0, const [0,6]}
       heel_r_ctrl.Location.Z {type=MovieSceneScriptingFloatChannel, default=0, const [0,6]}
       heel_r_ctrl.Rotation.X {type=MovieSceneScriptingFloatChannel, default=0, const [0,6]}
       heel_r_ctrl.Rotation.Y {type=MovieSceneScriptingFloatChannel, default=0, const [0,6]}
       heel_r_ctrl.Rotation.Z {type=MovieSceneScriptingFloatChannel, default=0, const [0,6]}
       heel_r_ctrl.Scale.X {type=MovieSceneScriptingFloatChannel, default=0, const [0,6]}
       heel_r_ctrl.Scale.Y {type=MovieSceneScriptingFloatChannel, default=0, const [0,6]}
       heel_r_ctrl.Scale.Z {type=MovieSceneScriptingFloatChannel, default=0, const [0,6]}
       leg_r_softness {type=MovieSceneScriptingFloatChannel, default=0, const [0,6]}
       leg_r_pv_ik_ctrl.Location.X {type=MovieSceneScriptingFloatChannel, default=0, keys [[324,0],[335,0.0256],[349,-0.0096],[351,-0.0097],[357,-0.0159],[362,0]]}
       leg_r_pv_ik_ctrl.Location.Y {type=MovieSceneScriptingFloatChannel, default=0, keys [[324,0],[335,-0.0127],[349,0.0266],[351,0.0267],[357,0.0559],[362,0]]}
       leg_r_pv_ik_ctrl.Location.Z {type=MovieSceneScriptingFloatChannel, default=0, keys [[324,0],[335,0.0348],[349,-0.0105],[351,-0.0106],[357,-0.0167],[362,0]]}
       leg_r_pv_ik_ctrl.Rotation.X {type=MovieSceneScriptingFloatChannel, default=0, keys [[324,0],[335,0.0322],[349,0.0453],[351,0.0454],[357,0.105],[362,0]]}
       leg_r_pv_ik_ctrl.Rotation.Y {type=MovieSceneScriptingFloatChannel, default=0, keys [[324,0],[335,0.0855],[349,0.0444],[351,0.0444],[357,0.1027],[362,0]]}
       leg_r_pv_ik_ctrl.Rotation.Z {type=MovieSceneScriptingFloatChannel, default=0, thinned_from=6, keys [[324,0],[335,-0.0038],[351,-0.0041],[357,-0.0095],[362,0]]}
       leg_r_pv_ik_ctrl.Scale.X {type=MovieSceneScriptingFloatChannel, default=0, const [0,6]}
       leg_r_pv_ik_ctrl.Scale.Y {type=MovieSceneScriptingFloatChannel, default=0, const [0,6]}
       leg_r_pv_ik_ctrl.Scale.Z {type=MovieSceneScriptingFloatChannel, default=0, const [0,6]}
       upperarm_l_fk_ctrl.Location.X {type=MovieSceneScriptingFloatChannel, default=0, const [0,9]}
       upperarm_l_fk_ctrl.Location.Y {type=MovieSceneScriptingFloatChannel, default=0, const [0,9]}
       upperarm_l_fk_ctrl.Location.Z {type=MovieSceneScriptingFloatChannel, default=0, const [0,9]}
       upperarm_l_fk_ctrl.Rotation.X {type=MovieSceneScriptingFloatChannel, default=0, thinned_from=9, keys [[100,0],[109,-24.9879],[132,0],[362,0]]}
       upperarm_l_fk_ctrl.Rotation.Y {type=MovieSceneScriptingFloatChannel, default=0, thinned_from=9, keys [[100,0],[109,9.7607],[132,0],[362,0]]}
       upperarm_l_fk_ctrl.Rotation.Z {type=MovieSceneScriptingFloatChannel, default=0, thinned_from=9, keys [[100,0],[109,-5.5971],[132,0],[362,0]]}
       upperarm_l_fk_ctrl.Scale.X {type=MovieSceneScriptingFloatChannel, default=0, const [0,9]}
       upperarm_l_fk_ctrl.Scale.Y {type=MovieSceneScriptingFloatChannel, default=0, const [0,9]}
       upperarm_l_fk_ctrl.Scale.Z {type=MovieSceneScriptingFloatChannel, default=0, const [0,9]}
       lowerarm_l_fk_ctrl.Location.X {type=MovieSceneScriptingFloatChannel, default=0, const [0,9]}
       lowerarm_l_fk_ctrl.Location.Y {type=MovieSceneScriptingFloatChannel, default=0, const [0,9]}
       lowerarm_l_fk_ctrl.Location.Z {type=MovieSceneScriptingFloatChannel, default=0, const [0,9]}
       lowerarm_l_fk_ctrl.Rotation.X {type=MovieSceneScriptingFloatChannel, default=0, thinned_from=9, keys [[100,0],[109,-1.0176],[132,0],[362,0]]}
       lowerarm_l_fk_ctrl.Rotation.Y {type=MovieSceneScriptingFloatChannel, default=0, thinned_from=9, keys [[100,0],[109,-0.6832],[132,0],[362,0]]}
       lowerarm_l_fk_ctrl.Rotation.Z {type=MovieSceneScriptingFloatChannel, default=0, thinned_from=9, keys [[100,0],[109,14.7746],[132,0],[362,0]]}
       lowerarm_l_fk_ctrl.Scale.X {type=MovieSceneScriptingFloatChannel, default=0, const [0,9]}
       lowerarm_l_fk_ctrl.Scale.Y {type=MovieSceneScriptingFloatChannel, default=0, const [0,9]}
       lowerarm_l_fk_ctrl.Scale.Z {type=MovieSceneScriptingFloatChannel, default=0, const [0,9]}
       hand_l_fk_ctrl.Location.X {type=MovieSceneScriptingFloatChannel, default=0, const [0,6]}
       hand_l_fk_ctrl.Location.Y {type=MovieSceneScriptingFloatChannel, default=0, const [0,6]}
       hand_l_fk_ctrl.Location.Z {type=MovieSceneScriptingFloatChannel, default=0, const [0,6]}
       hand_l_fk_ctrl.Rotation.X {type=MovieSceneScriptingFloatChannel, default=0, const [0,6]}
       hand_l_fk_ctrl.Rotation.Y {type=MovieSceneScriptingFloatChannel, default=0, const [0,6]}
       hand_l_fk_ctrl.Rotation.Z {type=MovieSceneScriptingFloatChannel, default=0, const [0,6]}
       hand_l_fk_ctrl.Scale.X {type=MovieSceneScriptingFloatChannel, default=0, const [0,6]}
       hand_l_fk_ctrl.Scale.Y {type=MovieSceneScriptingFloatChannel, default=0, const [0,6]}
       hand_l_fk_ctrl.Scale.Z {type=MovieSceneScriptingFloatChannel, default=0, const [0,6]}
       thumb_01_l_ctrl.Location.X {type=MovieSceneScriptingFloatChannel, default=0, const [0,6]}
       thumb_01_l_ctrl.Location.Y {type=MovieSceneScriptingFloatChannel, default=0, const [0,6]}
       thumb_01_l_ctrl.Location.Z {type=MovieSceneScriptingFloatChannel, default=0, const [0,6]}
       thumb_01_l_ctrl.Rotation.X {type=MovieSceneScriptingFloatChannel, default=0, const [0,6]}
       thumb_01_l_ctrl.Rotation.Y {type=MovieSceneScriptingFloatChannel, default=0, const [0,6]}
       thumb_01_l_ctrl.Rotation.Z {type=MovieSceneScriptingFloatChannel, default=0, const [0,6]}
       thumb_01_l_ctrl.Scale.X {type=MovieSceneScriptingFloatChannel, default=0, const [0,6]}
       thumb_01_l_ctrl.Scale.Y {type=MovieSceneScriptingFloatChannel, default=0, const [0,6]}
       thumb_01_l_ctrl.Scale.Z {type=MovieSceneScriptingFloatChannel, default=0, const [0,6]}
       thumb_02_l_ctrl.Location.X {type=MovieSceneScriptingFloatChannel, default=0, const [0,6]}
       thumb_02_l_ctrl.Location.Y {type=MovieSceneScriptingFloatChannel, default=0, const [0,6]}
       thumb_02_l_ctrl.Location.Z {type=MovieSceneScriptingFloatChannel, default=0, const [0,6]}
       thumb_02_l_ctrl.Rotation.X {type=MovieSceneScriptingFloatChannel, default=0, const [0,6]}
       thumb_02_l_ctrl.Rotation.Y {type=MovieSceneScriptingFloatChannel, default=0, const [0,6]}
       thumb_02_l_ctrl.Rotation.Z {type=MovieSceneScriptingFloatChannel, default=0, const [0,6]}
       thumb_02_l_ctrl.Scale.X {type=MovieSceneScriptingFloatChannel, default=0, const [0,6]}
       thumb_02_l_ctrl.Scale.Y {type=MovieSceneScriptingFloatChannel, default=0, const [0,6]}
       thumb_02_l_ctrl.Scale.Z {type=MovieSceneScriptingFloatChannel, default=0, const [0,6]}
       thumb_03_l_ctrl.Location.X {type=MovieSceneScriptingFloatChannel, default=0, const [0,6]}
       thumb_03_l_ctrl.Location.Y {type=MovieSceneScriptingFloatChannel, default=0, const [0,6]}
       thumb_03_l_ctrl.Location.Z {type=MovieSceneScriptingFloatChannel, default=0, thinned_from=6, keys [[324,0],[362,0]]}
       thumb_03_l_ctrl.Rotation.X {type=MovieSceneScriptingFloatChannel, default=0, const [0,6]}
       thumb_03_l_ctrl.Rotation.Y {type=MovieSceneScriptingFloatChannel, default=0, const [0,6]}
       thumb_03_l_ctrl.Rotation.Z {type=MovieSceneScriptingFloatChannel, default=0, const [0,6]}
       thumb_03_l_ctrl.Scale.X {type=MovieSceneScriptingFloatChannel, default=0, const [0,6]}
       thumb_03_l_ctrl.Scale.Y {type=MovieSceneScriptingFloatChannel, default=0, const [0,6]}
       thumb_03_l_ctrl.Scale.Z {type=MovieSceneScriptingFloatChannel, default=0, const [0,6]}
       index_metacarpal_l_ctrl.Location.X {type=MovieSceneScriptingFloatChannel, default=0, const [0,6]}
       index_metacarpal_l_ctrl.Location.Y {type=MovieSceneScriptingFloatChannel, default=0, const [0,6]}
       index_metacarpal_l_ctrl.Location.Z {type=MovieSceneScriptingFloatChannel, default=0, thinned_from=6, keys [[324,0],[362,0]]}
       index_metacarpal_l_ctrl.Rotation.X {type=MovieSceneScriptingFloatChannel, default=0, const [0,6]}
       index_metacarpal_l_ctrl.Rotation.Y {type=MovieSceneScriptingFloatChannel, default=0, const [0,6]}
       index_metacarpal_l_ctrl.Rotation.Z {type=MovieSceneScriptingFloatChannel, default=0, const [0,6]}
       index_metacarpal_l_ctrl.Scale.X {type=MovieSceneScriptingFloatChannel, default=0, const [0,6]}
       index_metacarpal_l_ctrl.Scale.Y {type=MovieSceneScriptingFloatChannel, default=0, const [0,6]}
       index_metacarpal_l_ctrl.Scale.Z {type=MovieSceneScriptingFloatChannel, default=0, const [0,6]}
       index_01_l_ctrl.Location.X {type=MovieSceneScriptingFloatChannel, default=0, const [0,6]}
       index_01_l_ctrl.Location.Y {type=MovieSceneScriptingFloatChannel, default=0, thinned_from=6, keys [[324,0],[362,0]]}
       index_01_l_ctrl.Location.Z {type=MovieSceneScriptingFloatChannel, default=0, const [0,6]}
       index_01_l_ctrl.Rotation.X {type=MovieSceneScriptingFloatChannel, default=0, const [0,6]}
       index_01_l_ctrl.Rotation.Y {type=MovieSceneScriptingFloatChannel, default=0, const [0,6]}
       index_01_l_ctrl.Rotation.Z {type=MovieSceneScriptingFloatChannel, default=0, const [0,6]}
       index_01_l_ctrl.Scale.X {type=MovieSceneScriptingFloatChannel, default=0, const [0,6]}
       index_01_l_ctrl.Scale.Y {type=MovieSceneScriptingFloatChannel, default=0, const [0,6]}
       index_01_l_ctrl.Scale.Z {type=MovieSceneScriptingFloatChannel, default=0, const [0,6]}
       index_02_l_ctrl.Location.X {type=MovieSceneScriptingFloatChannel, default=0, const [0,6]}
       index_02_l_ctrl.Location.Y {type=MovieSceneScriptingFloatChannel, default=0, const [0,6]}
       index_02_l_ctrl.Location.Z {type=MovieSceneScriptingFloatChannel, default=0, const [0,6]}
       index_02_l_ctrl.Rotation.X {type=MovieSceneScriptingFloatChannel, default=0, const [0,6]}
       index_02_l_ctrl.Rotation.Y {type=MovieSceneScriptingFloatChannel, default=0, const [0,6]}
       index_02_l_ctrl.Rotation.Z {type=MovieSceneScriptingFloatChannel, default=0, const [0,6]}
       index_02_l_ctrl.Scale.X {type=MovieSceneScriptingFloatChannel, default=0, const [0,6]}
       index_02_l_ctrl.Scale.Y {type=MovieSceneScriptingFloatChannel, default=0, const [0,6]}
       index_02_l_ctrl.Scale.Z {type=MovieSceneScriptingFloatChannel, default=0, const [0,6]}
       index_03_l_ctrl.Location.X {type=MovieSceneScriptingFloatChannel, default=0, const [0,6]}
       index_03_l_ctrl.Location.Y {type=MovieSceneScriptingFloatChannel, default=0, const [0,6]}
       index_03_l_ctrl.Location.Z {type=MovieSceneScriptingFloatChannel, default=0, const [0,6]}
       index_03_l_ctrl.Rotation.X {type=MovieSceneScriptingFloatChannel, default=0, const [0,6]}
       index_03_l_ctrl.Rotation.Y {type=MovieSceneScriptingFloatChannel, default=0, const [0,6]}
       index_03_l_ctrl.Rotation.Z {type=MovieSceneScriptingFloatChannel, default=0, const [0,6]}
       index_03_l_ctrl.Scale.X {type=MovieSceneScriptingFloatChannel, default=0, const [0,6]}
       index_03_l_ctrl.Scale.Y {type=MovieSceneScriptingFloatChannel, default=0, const [0,6]}
       index_03_l_ctrl.Scale.Z {type=MovieSceneScriptingFloatChannel, default=0, const [0,6]}
       middle_metacarpal_l_ctrl.Location.X {type=MovieSceneScriptingFloatChannel, default=0, const [0,6]}
       middle_metacarpal_l_ctrl.Location.Y {type=MovieSceneScriptingFloatChannel, default=0, const [0,6]}
       middle_metacarpal_l_ctrl.Location.Z {type=MovieSceneScriptingFloatChannel, default=0, const [0,6]}
       middle_metacarpal_l_ctrl.Rotation.X {type=MovieSceneScriptingFloatChannel, default=0, const [0,6]}
       middle_metacarpal_l_ctrl.Rotation.Y {type=MovieSceneScriptingFloatChannel, default=0, const [0,6]}
       middle_metacarpal_l_ctrl.Rotation.Z {type=MovieSceneScriptingFloatChannel, default=0, const [0,6]}
       middle_metacarpal_l_ctrl.Scale.X {type=MovieSceneScriptingFloatChannel, default=0, const [0,6]}
       middle_metacarpal_l_ctrl.Scale.Y {type=MovieSceneScriptingFloatChannel, default=0, const [0,6]}
       middle_metacarpal_l_ctrl.Scale.Z {type=MovieSceneScriptingFloatChannel, default=0, const [0,6]}
       middle_01_l_ctrl.Location.X {type=MovieSceneScriptingFloatChannel, default=0, const [0,6]}
       middle_01_l_ctrl.Location.Y {type=MovieSceneScriptingFloatChannel, default=0, const [0,6]}
       middle_01_l_ctrl.Location.Z {type=MovieSceneScriptingFloatChannel, default=0, const [0,6]}
       middle_01_l_ctrl.Rotation.X {type=MovieSceneScriptingFloatChannel, default=0, const [0,6]}
       middle_01_l_ctrl.Rotation.Y {type=MovieSceneScriptingFloatChannel, default=0, const [0,6]}
       middle_01_l_ctrl.Rotation.Z {type=MovieSceneScriptingFloatChannel, default=0, const [0,6]}
       middle_01_l_ctrl.Scale.X {type=MovieSceneScriptingFloatChannel, default=0, const [0,6]}
       middle_01_l_ctrl.Scale.Y {type=MovieSceneScriptingFloatChannel, default=0, const [0,6]}
       middle_01_l_ctrl.Scale.Z {type=MovieSceneScriptingFloatChannel, default=0, const [0,6]}
       middle_02_l_ctrl.Location.X {type=MovieSceneScriptingFloatChannel, default=0, const [0,6]}
       middle_02_l_ctrl.Location.Y {type=MovieSceneScriptingFloatChannel, default=0, const [0,6]}
       middle_02_l_ctrl.Location.Z {type=MovieSceneScriptingFloatChannel, default=0, const [0,6]}
       middle_02_l_ctrl.Rotation.X {type=MovieSceneScriptingFloatChannel, default=0, const [0,6]}
       middle_02_l_ctrl.Rotation.Y {type=MovieSceneScriptingFloatChannel, default=0, const [0,6]}
       middle_02_l_ctrl.Rotation.Z {type=MovieSceneScriptingFloatChannel, default=0, const [0,6]}
       middle_02_l_ctrl.Scale.X {type=MovieSceneScriptingFloatChannel, default=0, const [0,6]}
       middle_02_l_ctrl.Scale.Y {type=MovieSceneScriptingFloatChannel, default=0, const [0,6]}
       middle_02_l_ctrl.Scale.Z {type=MovieSceneScriptingFloatChannel, default=0, const [0,6]}
       middle_03_l_ctrl.Location.X {type=MovieSceneScriptingFloatChannel, default=0, const [0,6]}
       middle_03_l_ctrl.Location.Y {type=MovieSceneScriptingFloatChannel, default=0, const [0,6]}
       middle_03_l_ctrl.Location.Z {type=MovieSceneScriptingFloatChannel, default=0, const [0,6]}
       middle_03_l_ctrl.Rotation.X {type=MovieSceneScriptingFloatChannel, default=0, const [0,6]}
       middle_03_l_ctrl.Rotation.Y {type=MovieSceneScriptingFloatChannel, default=0, const [0,6]}
       middle_03_l_ctrl.Rotation.Z {type=MovieSceneScriptingFloatChannel, default=0, const [0,6]}
       middle_03_l_ctrl.Scale.X {type=MovieSceneScriptingFloatChannel, default=0, const [0,6]}
       middle_03_l_ctrl.Scale.Y {type=MovieSceneScriptingFloatChannel, default=0, const [0,6]}
       middle_03_l_ctrl.Scale.Z {type=MovieSceneScriptingFloatChannel, default=0, const [0,6]}
       ring_metacarpal_l_ctrl.Location.X {type=MovieSceneScriptingFloatChannel, default=0, thinned_from=6, keys [[324,0],[362,0]]}
       ring_metacarpal_l_ctrl.Location.Y {type=MovieSceneScriptingFloatChannel, default=0, thinned_from=6, keys [[324,0],[362,0]]}
       ring_metacarpal_l_ctrl.Location.Z {type=MovieSceneScriptingFloatChannel, default=0, thinned_from=6, keys [[324,0],[362,0]]}
       ring_metacarpal_l_ctrl.Rotation.X {type=MovieSceneScriptingFloatChannel, default=0, const [0,6]}
       ring_metacarpal_l_ctrl.Rotation.Y {type=MovieSceneScriptingFloatChannel, default=0, const [0,6]}
       ring_metacarpal_l_ctrl.Rotation.Z {type=MovieSceneScriptingFloatChannel, default=0, const [0,6]}
       ring_metacarpal_l_ctrl.Scale.X {type=MovieSceneScriptingFloatChannel, default=0, const [0,6]}
       ring_metacarpal_l_ctrl.Scale.Y {type=MovieSceneScriptingFloatChannel, default=0, const [0,6]}
       ring_metacarpal_l_ctrl.Scale.Z {type=MovieSceneScriptingFloatChannel, default=0, const [0,6]}
       ring_01_l_ctrl.Location.X {type=MovieSceneScriptingFloatChannel, default=0, const [0,6]}
       ring_01_l_ctrl.Location.Y {type=MovieSceneScriptingFloatChannel, default=0, const [0,6]}
       ring_01_l_ctrl.Location.Z {type=MovieSceneScriptingFloatChannel, default=0, const [0,6]}
       ring_01_l_ctrl.Rotation.X {type=MovieSceneScriptingFloatChannel, default=0, const [0,6]}
       ring_01_l_ctrl.Rotation.Y {type=MovieSceneScriptingFloatChannel, default=0, const [0,6]}
       ring_01_l_ctrl.Rotation.Z {type=MovieSceneScriptingFloatChannel, default=0, const [0,6]}
       ring_01_l_ctrl.Scale.X {type=MovieSceneScriptingFloatChannel, default=0, const [0,6]}
       ring_01_l_ctrl.Scale.Y {type=MovieSceneScriptingFloatChannel, default=0, const [0,6]}
       ring_01_l_ctrl.Scale.Z {type=MovieSceneScriptingFloatChannel, default=0, const [0,6]}
       ring_02_l_ctrl.Location.X {type=MovieSceneScriptingFloatChannel, default=0, const [0,6]}
       ring_02_l_ctrl.Location.Y {type=MovieSceneScriptingFloatChannel, default=0, const [0,6]}
       ring_02_l_ctrl.Location.Z {type=MovieSceneScriptingFloatChannel, default=0, const [0,6]}
       ring_02_l_ctrl.Rotation.X {type=MovieSceneScriptingFloatChannel, default=0, const [0,6]}
       ring_02_l_ctrl.Rotation.Y {type=MovieSceneScriptingFloatChannel, default=0, const [0,6]}
       ring_02_l_ctrl.Rotation.Z {type=MovieSceneScriptingFloatChannel, default=0, const [0,6]}
       ring_02_l_ctrl.Scale.X {type=MovieSceneScriptingFloatChannel, default=0, const [0,6]}
       ring_02_l_ctrl.Scale.Y {type=MovieSceneScriptingFloatChannel, default=0, const [0,6]}
       ring_02_l_ctrl.Scale.Z {type=MovieSceneScriptingFloatChannel, default=0, const [0,6]}
       ring_03_l_ctrl.Location.X {type=MovieSceneScriptingFloatChannel, default=0, const [0,6]}
       ring_03_l_ctrl.Location.Y {type=MovieSceneScriptingFloatChannel, default=0, const [0,6]}
       ring_03_l_ctrl.Location.Z {type=MovieSceneScriptingFloatChannel, default=0, const [0,6]}
       ring_03_l_ctrl.Rotation.X {type=MovieSceneScriptingFloatChannel, default=0, const [0,6]}
       ring_03_l_ctrl.Rotation.Y {type=MovieSceneScriptingFloatChannel, default=0, const [0,6]}
       ring_03_l_ctrl.Rotation.Z {type=MovieSceneScriptingFloatChannel, default=0, const [0,6]}
       ring_03_l_ctrl.Scale.X {type=MovieSceneScriptingFloatChannel, default=0, const [0,6]}
       ring_03_l_ctrl.Scale.Y {type=MovieSceneScriptingFloatChannel, default=0, const [0,6]}
       ring_03_l_ctrl.Scale.Z {type=MovieSceneScriptingFloatChannel, default=0, const [0,6]}
       pinky_metacarpal_l_ctrl.Location.X {type=MovieSceneScriptingFloatChannel, default=0, const [0,6]}
       pinky_metacarpal_l_ctrl.Location.Y {type=MovieSceneScriptingFloatChannel, default=0, thinned_from=6, keys [[324,0],[362,0]]}
       pinky_metacarpal_l_ctrl.Location.Z {type=MovieSceneScriptingFloatChannel, default=0, const [0,6]}
       pinky_metacarpal_l_ctrl.Rotation.X {type=MovieSceneScriptingFloatChannel, default=0, const [0,6]}
       pinky_metacarpal_l_ctrl.Rotation.Y {type=MovieSceneScriptingFloatChannel, default=0, const [0,6]}
       pinky_metacarpal_l_ctrl.Rotation.Z {type=MovieSceneScriptingFloatChannel, default=0, const [0,6]}
       pinky_metacarpal_l_ctrl.Scale.X {type=MovieSceneScriptingFloatChannel, default=0, const [0,6]}
       pinky_metacarpal_l_ctrl.Scale.Y {type=MovieSceneScriptingFloatChannel, default=0, const [0,6]}
       pinky_metacarpal_l_ctrl.Scale.Z {type=MovieSceneScriptingFloatChannel, default=0, const [0,6]}
       pinky_01_l_ctrl.Location.X {type=MovieSceneScriptingFloatChannel, default=0, const [0,6]}
       pinky_01_l_ctrl.Location.Y {type=MovieSceneScriptingFloatChannel, default=0, const [0,6]}
       pinky_01_l_ctrl.Location.Z {type=MovieSceneScriptingFloatChannel, default=0, thinned_from=6, keys [[324,0],[362,0]]}
       pinky_01_l_ctrl.Rotation.X {type=MovieSceneScriptingFloatChannel, default=0, const [0,6]}
       pinky_01_l_ctrl.Rotation.Y {type=MovieSceneScriptingFloatChannel, default=0, const [0,6]}
       pinky_01_l_ctrl.Rotation.Z {type=MovieSceneScriptingFloatChannel, default=0, const [0,6]}
       pinky_01_l_ctrl.Scale.X {type=MovieSceneScriptingFloatChannel, default=0, const [0,6]}
       pinky_01_l_ctrl.Scale.Y {type=MovieSceneScriptingFloatChannel, default=0, const [0,6]}
       pinky_01_l_ctrl.Scale.Z {type=MovieSceneScriptingFloatChannel, default=0, const [0,6]}
       pinky_02_l_ctrl.Location.X {type=MovieSceneScriptingFloatChannel, default=0, thinned_from=6, keys [[324,0],[362,0]]}
       pinky_02_l_ctrl.Location.Y {type=MovieSceneScriptingFloatChannel, default=0, const [0,6]}
       pinky_02_l_ctrl.Location.Z {type=MovieSceneScriptingFloatChannel, default=0, const [0,6]}
       pinky_02_l_ctrl.Rotation.X {type=MovieSceneScriptingFloatChannel, default=0, const [0,6]}
       pinky_02_l_ctrl.Rotation.Y {type=MovieSceneScriptingFloatChannel, default=0, const [0,6]}
       pinky_02_l_ctrl.Rotation.Z {type=MovieSceneScriptingFloatChannel, default=0, const [0,6]}
       pinky_02_l_ctrl.Scale.X {type=MovieSceneScriptingFloatChannel, default=0, const [0,6]}
       pinky_02_l_ctrl.Scale.Y {type=MovieSceneScriptingFloatChannel, default=0, const [0,6]}
       pinky_02_l_ctrl.Scale.Z {type=MovieSceneScriptingFloatChannel, default=0, const [0,6]}
       pinky_03_l_ctrl.Location.X {type=MovieSceneScriptingFloatChannel, default=0, thinned_from=6, keys [[324,0],[362,0]]}
       pinky_03_l_ctrl.Location.Y {type=MovieSceneScriptingFloatChannel, default=0, const [0,6]}
       pinky_03_l_ctrl.Location.Z {type=MovieSceneScriptingFloatChannel, default=0, const [0,6]}
       pinky_03_l_ctrl.Rotation.X {type=MovieSceneScriptingFloatChannel, default=0, const [0,6]}
       pinky_03_l_ctrl.Rotation.Y {type=MovieSceneScriptingFloatChannel, default=0, const [0,6]}
       pinky_03_l_ctrl.Rotation.Z {type=MovieSceneScriptingFloatChannel, default=0, const [0,6]}
       pinky_03_l_ctrl.Scale.X {type=MovieSceneScriptingFloatChannel, default=0, const [0,6]}
       pinky_03_l_ctrl.Scale.Y {type=MovieSceneScriptingFloatChannel, default=0, const [0,6]}
       pinky_03_l_ctrl.Scale.Z {type=MovieSceneScriptingFloatChannel, default=0, const [0,6]}
       hand_l_ik_ctrl.Location.X {type=MovieSceneScriptingFloatChannel, default=0, keys [[324,-0.9191],[335,-1.0651],[349,0.0231],[351,0.0243],[357,0.0186],[362,0.0283]]}
       hand_l_ik_ctrl.Location.Y {type=MovieSceneScriptingFloatChannel, default=0, keys [[324,3.3791],[335,-6.1718],[349,0.0609],[351,0.0719],[357,0.1066],[362,0.0082]]}
       hand_l_ik_ctrl.Location.Z {type=MovieSceneScriptingFloatChannel, default=0, keys [[324,7.8849],[335,5.2071],[349,-0.0021],[351,-0.0062],[357,0.0269],[362,-0.0323]]}
       hand_l_ik_ctrl.Rotation.X {type=MovieSceneScriptingFloatChannel, default=0, keys [[324,6.0268],[335,-1.8377],[349,-0.0338],[351,-0.0319],[357,-0.0738],[362,0]]}
       hand_l_ik_ctrl.Rotation.Y {type=MovieSceneScriptingFloatChannel, default=0, keys [[324,19.8732],[335,14.9964],[349,0.1798],[351,0.1675],[357,0.3876],[362,0]]}
       hand_l_ik_ctrl.Rotation.Z {type=MovieSceneScriptingFloatChannel, default=0, keys [[324,12.4162],[335,-5.7912],[349,-0.2415],[351,-0.2357],[357,-0.5456],[362,0]]}
       hand_l_ik_ctrl.Scale.X {type=MovieSceneScriptingFloatChannel, default=0, const [0,6]}
       hand_l_ik_ctrl.Scale.Y {type=MovieSceneScriptingFloatChannel, default=0, const [0,6]}
       hand_l_ik_ctrl.Scale.Z {type=MovieSceneScriptingFloatChannel, default=0, const [0,6]}
       Softness {type=MovieSceneScriptingFloatChannel, default=0, const [0,6]}
       hand_l_ikorient_ctrl.Rotation.Roll {type=MovieSceneScriptingFloatChannel, default=0, const [0,6]}
       hand_l_ikorient_ctrl.Rotation.Pitch {type=MovieSceneScriptingFloatChannel, default=0, const [0,6]}
       hand_l_ikorient_ctrl.Rotation.Yaw {type=MovieSceneScriptingFloatChannel, default=0, const [0,6]}
       arm_l_pv_ik_ctrl.Location.X {type=MovieSceneScriptingFloatChannel, default=0, keys [[324,0.0006],[335,0.9047],[349,0.0069],[351,0.006],[357,0.0097],[362,0.0061]]}
       arm_l_pv_ik_ctrl.Location.Y {type=MovieSceneScriptingFloatChannel, default=0, keys [[324,0.0064],[335,-1.9062],[349,-0.0045],[351,-0.0013],[357,0.007],[362,-0.0167]]}
       arm_l_pv_ik_ctrl.Location.Z {type=MovieSceneScriptingFloatChannel, default=0, keys [[324,-0.0052],[335,1.8394],[349,-0.0751],[351,-0.0792],[357,-0.1013],[362,-0.04]]}
       arm_l_pv_ik_ctrl.Rotation.X {type=MovieSceneScriptingFloatChannel, default=0, keys [[324,0],[335,-1.0748],[349,0.0567],[351,0.0593],[357,0.0934],[362,0]]}
       arm_l_pv_ik_ctrl.Rotation.Y {type=MovieSceneScriptingFloatChannel, default=0, keys [[324,0],[335,8.5292],[349,0.0289],[351,0.0199],[357,0.0462],[362,0]]}
       arm_l_pv_ik_ctrl.Rotation.Z {type=MovieSceneScriptingFloatChannel, default=0, keys [[324,0],[335,-10.1238],[349,-0.0136],[351,-0.003],[357,-0.0069],[362,0]]}
       arm_l_pv_ik_ctrl.Scale.X {type=MovieSceneScriptingFloatChannel, default=0, const [0,6]}
       arm_l_pv_ik_ctrl.Scale.Y {type=MovieSceneScriptingFloatChannel, default=0, const [0,6]}
       arm_l_pv_ik_ctrl.Scale.Z {type=MovieSceneScriptingFloatChannel, default=0, const [0,6]}
       upperarm_r_fk_ctrl.Location.X {type=MovieSceneScriptingFloatChannel, default=0, const [0,11]}
       upperarm_r_fk_ctrl.Location.Y {type=MovieSceneScriptingFloatChannel, default=0, const [0,11]}
       upperarm_r_fk_ctrl.Location.Z {type=MovieSceneScriptingFloatChannel, default=0, const [0,11]}
       upperarm_r_fk_ctrl.Rotation.X {type=MovieSceneScriptingFloatChannel, default=0, const [0,11]}
       upperarm_r_fk_ctrl.Rotation.Y {type=MovieSceneScriptingFloatChannel, default=0, const [0,11]}
       upperarm_r_fk_ctrl.Rotation.Z {type=MovieSceneScriptingFloatChannel, default=0, const [0,11]}
       upperarm_r_fk_ctrl.Scale.X {type=MovieSceneScriptingFloatChannel, default=0, const [0,11]}
       upperarm_r_fk_ctrl.Scale.Y {type=MovieSceneScriptingFloatChannel, default=0, const [0,11]}
       upperarm_r_fk_ctrl.Scale.Z {type=MovieSceneScriptingFloatChannel, default=0, const [0,11]}
       lowerarm_r_fk_ctrl.Location.X {type=MovieSceneScriptingFloatChannel, default=0, const [0,14]}
       lowerarm_r_fk_ctrl.Location.Y {type=MovieSceneScriptingFloatChannel, default=0, const [0,14]}
       lowerarm_r_fk_ctrl.Location.Z {type=MovieSceneScriptingFloatChannel, default=0, const [0,14]}
       lowerarm_r_fk_ctrl.Rotation.X {type=MovieSceneScriptingFloatChannel, default=0, thinned_from=14, keys [[91,0],[108,-0.9472],[132,0],[362,0]]}
       lowerarm_r_fk_ctrl.Rotation.Y {type=MovieSceneScriptingFloatChannel, default=0, thinned_from=14, keys [[91,0],[108,-0.7786],[132,0],[362,0]]}
       lowerarm_r_fk_ctrl.Rotation.Z {type=MovieSceneScriptingFloatChannel, default=0, thinned_from=14, keys [[91,0],[108,16.9475],[132,0],[362,0]]}
       lowerarm_r_fk_ctrl.Scale.X {type=MovieSceneScriptingFloatChannel, default=0, const [0,14]}
       lowerarm_r_fk_ctrl.Scale.Y {type=MovieSceneScriptingFloatChannel, default=0, const [0,14]}
       lowerarm_r_fk_ctrl.Scale.Z {type=MovieSceneScriptingFloatChannel, default=0, const [0,14]}
       hand_r_fk_ctrl.Location.X {type=MovieSceneScriptingFloatChannel, default=0, const [0,12]}
       hand_r_fk_ctrl.Location.Y {type=MovieSceneScriptingFloatChannel, default=0, const [0,12]}
       hand_r_fk_ctrl.Location.Z {type=MovieSceneScriptingFloatChannel, default=0, const [0,12]}
       hand_r_fk_ctrl.Rotation.X {type=MovieSceneScriptingFloatChannel, default=0, thinned_from=12, keys [[212,0],[335,0],[336,-1.5717],[341,-21.2178],[349,0],[362,0]]}
       hand_r_fk_ctrl.Rotation.Y {type=MovieSceneScriptingFloatChannel, default=0, thinned_from=12, keys [[212,0],[335,0],[336,0.3143],[341,4.2433],[349,0],[362,0]]}
       hand_r_fk_ctrl.Rotation.Z {type=MovieSceneScriptingFloatChannel, default=0, thinned_from=12, keys [[212,0],[335,0],[336,1.5035],[341,20.2979],[349,0],[362,0]]}
       hand_r_fk_ctrl.Scale.X {type=MovieSceneScriptingFloatChannel, default=0, const [0,12]}
       hand_r_fk_ctrl.Scale.Y {type=MovieSceneScriptingFloatChannel, default=0, const [0,12]}
       hand_r_fk_ctrl.Scale.Z {type=MovieSceneScriptingFloatChannel, default=0, const [0,12]}
       thumb_01_r_ctrl.Location.X {type=MovieSceneScriptingFloatChannel, default=0, thinned_from=12, keys [[212,0],[362,0]]}
       thumb_01_r_ctrl.Location.Y {type=MovieSceneScriptingFloatChannel, default=0, thinned_from=12, keys [[212,0],[362,0]]}
       thumb_01_r_ctrl.Location.Z {type=MovieSceneScriptingFloatChannel, default=0, thinned_from=12, keys [[212,0],[362,0]]}
       thumb_01_r_ctrl.Rotation.X {
        type=MovieSceneScriptingFloatChannel
        default=0
        thinned_from=12
        keys [[212,0],[335,0],[336,-0.9741],[341,-13.15],[349,-0.1374],[351,0],[362,0]]
       }
       thumb_01_r_ctrl.Rotation.Y {
        type=MovieSceneScriptingFloatChannel
        default=0
        thinned_from=12
        keys [[212,0],[335,0],[336,-0.2965],[341,-4.0034],[349,-0.0418],[351,0],[362,0]]
       }
       thumb_01_r_ctrl.Rotation.Z {
        type=MovieSceneScriptingFloatChannel
        default=0
        thinned_from=12
        keys [[212,0],[335,0],[336,0.7367],[341,9.9448],[349,0.1039],[351,0],[362,0]]
       }
       thumb_01_r_ctrl.Scale.X {type=MovieSceneScriptingFloatChannel, default=0, const [0,12]}
       thumb_01_r_ctrl.Scale.Y {type=MovieSceneScriptingFloatChannel, default=0, const [0,12]}
       thumb_01_r_ctrl.Scale.Z {type=MovieSceneScriptingFloatChannel, default=0, const [0,12]}
       thumb_02_r_ctrl.Location.X {type=MovieSceneScriptingFloatChannel, default=0, thinned_from=12, keys [[212,0],[336,0],[341,0.0002],[362,0]]}
       thumb_02_r_ctrl.Location.Y {type=MovieSceneScriptingFloatChannel, default=0, thinned_from=12, keys [[212,0],[362,0]]}
       thumb_02_r_ctrl.Location.Z {type=MovieSceneScriptingFloatChannel, default=0, thinned_from=12, keys [[212,0],[362,0]]}
       thumb_02_r_ctrl.Rotation.X {
        type=MovieSceneScriptingFloatChannel
        default=0
        thinned_from=12
        keys [[212,0],[335,0],[336,0.0114],[341,0.1543],[349,0.0016],[351,0],[362,0]]
       }
       thumb_02_r_ctrl.Rotation.Y {
        type=MovieSceneScriptingFloatChannel
        default=0
        thinned_from=12
        keys [[212,0],[335,0],[336,-0.0361],[341,-0.4877],[349,-0.0051],[351,0],[362,0]]
       }
       thumb_02_r_ctrl.Rotation.Z {
        type=MovieSceneScriptingFloatChannel
        default=0
        thinned_from=12
        keys [[212,0],[335,0],[336,0.3967],[341,5.3559],[349,0.056],[351,0],[362,0]]
       }
       thumb_02_r_ctrl.Scale.X {type=MovieSceneScriptingFloatChannel, default=0, const [0,12]}
       thumb_02_r_ctrl.Scale.Y {type=MovieSceneScriptingFloatChannel, default=0, const [0,12]}
       thumb_02_r_ctrl.Scale.Z {type=MovieSceneScriptingFloatChannel, default=0, const [0,12]}
       thumb_03_r_ctrl.Location.X {type=MovieSceneScriptingFloatChannel, default=0, thinned_from=12, keys [[212,0],[362,0]]}
       thumb_03_r_ctrl.Location.Y {type=MovieSceneScriptingFloatChannel, default=0, thinned_from=12, keys [[212,0],[362,0]]}
       thumb_03_r_ctrl.Location.Z {type=MovieSceneScriptingFloatChannel, default=0, thinned_from=12, keys [[212,0],[362,0]]}
       thumb_03_r_ctrl.Rotation.X {
        type=MovieSceneScriptingFloatChannel
        default=0
        thinned_from=12
        keys [[212,0],[335,0],[336,-0.0336],[341,-0.454],[349,-0.0047],[351,0],[362,0]]
       }
       thumb_03_r_ctrl.Rotation.Y {
        type=MovieSceneScriptingFloatChannel
        default=0
        thinned_from=12
        keys [[212,0],[335,0],[336,-0.047],[341,-0.634],[349,-0.0066],[351,0],[362,0]]
       }
       thumb_03_r_ctrl.Rotation.Z {
        type=MovieSceneScriptingFloatChannel
        default=0
        thinned_from=12
        keys [[212,0],[335,0],[336,1.1121],[341,15.0127],[349,0.1569],[351,0],[362,0]]
       }
       thumb_03_r_ctrl.Scale.X {type=MovieSceneScriptingFloatChannel, default=0, const [0,12]}
       thumb_03_r_ctrl.Scale.Y {type=MovieSceneScriptingFloatChannel, default=0, const [0,12]}
       thumb_03_r_ctrl.Scale.Z {type=MovieSceneScriptingFloatChannel, default=0, const [0,12]}
       index_metacarpal_r_ctrl.Location.X {type=MovieSceneScriptingFloatChannel, default=0, thinned_from=12, keys [[212,0],[362,0]]}
       index_metacarpal_r_ctrl.Location.Y {type=MovieSceneScriptingFloatChannel, default=0, thinned_from=12, keys [[212,0],[362,0]]}
       index_metacarpal_r_ctrl.Location.Z {type=MovieSceneScriptingFloatChannel, default=0, thinned_from=12, keys [[212,0],[362,0]]}
       index_metacarpal_r_ctrl.Rotation.X {type=MovieSceneScriptingFloatChannel, default=0, thinned_from=12, keys [[212,0],[362,0]]}
       index_metacarpal_r_ctrl.Rotation.Y {
        type=MovieSceneScriptingFloatChannel
        default=0
        thinned_from=12
        keys [[212,0],[335,0],[336,-0.1384],[341,-1.8678],[349,-0.0195],[351,0],[362,0]]
       }
       index_metacarpal_r_ctrl.Rotation.Z {type=MovieSceneScriptingFloatChannel, default=0, thinned_from=12, keys [[212,0],[362,0]]}
       index_metacarpal_r_ctrl.Scale.X {type=MovieSceneScriptingFloatChannel, default=0, const [0,12]}
       index_metacarpal_r_ctrl.Scale.Y {type=MovieSceneScriptingFloatChannel, default=0, const [0,12]}
       index_metacarpal_r_ctrl.Scale.Z {type=MovieSceneScriptingFloatChannel, default=0, const [0,12]}
       index_01_r_ctrl.Location.X {type=MovieSceneScriptingFloatChannel, default=0, thinned_from=12, keys [[212,0],[336,0],[341,0.0002],[362,0]]}
       index_01_r_ctrl.Location.Y {type=MovieSceneScriptingFloatChannel, default=0, thinned_from=12, keys [[212,0],[362,0]]}
       index_01_r_ctrl.Location.Z {type=MovieSceneScriptingFloatChannel, default=0, thinned_from=12, keys [[212,0],[341,-0.0001],[362,0]]}
       index_01_r_ctrl.Rotation.X {
        type=MovieSceneScriptingFloatChannel
        default=0
        thinned_from=12
        keys [[212,0],[335,0],[336,0.3179],[341,4.2921],[349,0.0449],[351,0],[362,0]]
       }
       index_01_r_ctrl.Rotation.Y {
        type=MovieSceneScriptingFloatChannel
        default=0
        thinned_from=12
        keys [[212,0],[335,0],[336,-0.9186],[341,-12.4008],[349,-0.1296],[351,0],[362,0]]
       }
       index_01_r_ctrl.Rotation.Z {
        type=MovieSceneScriptingFloatChannel
        default=0
        thinned_from=12
        keys [[212,0],[335,0],[336,1.3454],[341,18.1635],[349,0.1898],[351,0],[362,0]]
       }
       index_01_r_ctrl.Scale.X {type=MovieSceneScriptingFloatChannel, default=0, const [0,12]}
       index_01_r_ctrl.Scale.Y {type=MovieSceneScriptingFloatChannel, default=0, const [0,12]}
       index_01_r_ctrl.Scale.Z {type=MovieSceneScriptingFloatChannel, default=0, const [0,12]}
       index_02_r_ctrl.Location.X {type=MovieSceneScriptingFloatChannel, default=0, thinned_from=12, keys [[212,0],[362,0]]}
       index_02_r_ctrl.Location.Y {type=MovieSceneScriptingFloatChannel, default=0, thinned_from=12, keys [[212,0],[362,0]]}
       index_02_r_ctrl.Location.Z {type=MovieSceneScriptingFloatChannel, default=0, thinned_from=12, keys [[212,0],[362,0]]}
       index_02_r_ctrl.Rotation.X {
        type=MovieSceneScriptingFloatChannel
        default=0
        thinned_from=12
        keys [[212,0],[335,0],[336,0.0042],[341,0.0569],[349,0.0006],[351,0],[362,0]]
       }
       index_02_r_ctrl.Rotation.Y {
        type=MovieSceneScriptingFloatChannel
        default=0
        thinned_from=12
        keys [[212,0],[335,0],[336,0.0056],[341,0.0759],[349,0.0008],[351,0],[362,0]]
       }
       index_02_r_ctrl.Rotation.Z {
        type=MovieSceneScriptingFloatChannel
        default=0
        thinned_from=12
        keys [[212,0],[335,0],[336,1.0251],[341,13.8385],[349,0.1446],[351,0],[362,0]]
       }
       index_02_r_ctrl.Scale.X {type=MovieSceneScriptingFloatChannel, default=0, const [0,12]}
       index_02_r_ctrl.Scale.Y {type=MovieSceneScriptingFloatChannel, default=0, const [0,12]}
       index_02_r_ctrl.Scale.Z {type=MovieSceneScriptingFloatChannel, default=0, const [0,12]}
       index_03_r_ctrl.Location.X {type=MovieSceneScriptingFloatChannel, default=0, thinned_from=12, keys [[212,0],[336,0],[341,-0.0002],[362,0]]}
       index_03_r_ctrl.Location.Y {type=MovieSceneScriptingFloatChannel, default=0, thinned_from=12, keys [[212,0],[362,0]]}
       index_03_r_ctrl.Location.Z {type=MovieSceneScriptingFloatChannel, default=0, thinned_from=12, keys [[212,0],[362,0]]}
       index_03_r_ctrl.Rotation.X {
        type=MovieSceneScriptingFloatChannel
        default=0
        thinned_from=12
        keys [[212,0],[335,0],[336,0.0199],[341,0.2682],[349,0.0028],[351,0],[362,0]]
       }
       index_03_r_ctrl.Rotation.Y {
        type=MovieSceneScriptingFloatChannel
        default=0
        thinned_from=12
        keys [[212,0],[335,0],[336,0.0055],[341,0.0738],[349,0.0008],[351,0],[362,0]]
       }
       index_03_r_ctrl.Rotation.Z {
        type=MovieSceneScriptingFloatChannel
        default=0
        thinned_from=12
        keys [[212,0],[335,0],[336,1.4939],[341,20.1673],[349,0.2108],[351,0],[362,0]]
       }
       index_03_r_ctrl.Scale.X {type=MovieSceneScriptingFloatChannel, default=0, const [0,12]}
       index_03_r_ctrl.Scale.Y {type=MovieSceneScriptingFloatChannel, default=0, const [0,12]}
       index_03_r_ctrl.Scale.Z {type=MovieSceneScriptingFloatChannel, default=0, const [0,12]}
       middle_metacarpal_r_ctrl.Location.X {type=MovieSceneScriptingFloatChannel, default=0, thinned_from=12, keys [[212,0],[362,0]]}
       middle_metacarpal_r_ctrl.Location.Y {type=MovieSceneScriptingFloatChannel, default=0, thinned_from=12, keys [[212,0],[362,0]]}
       middle_metacarpal_r_ctrl.Location.Z {type=MovieSceneScriptingFloatChannel, default=0, thinned_from=12, keys [[212,0],[362,0]]}
       middle_metacarpal_r_ctrl.Rotation.X {type=MovieSceneScriptingFloatChannel, default=0, thinned_from=12, keys [[212,0],[362,0]]}
       middle_metacarpal_r_ctrl.Rotation.Y {
        type=MovieSceneScriptingFloatChannel
        default=0
        thinned_from=12
        keys [[212,0],[335,0],[336,-0.001],[341,-0.0135],[349,-0.0001],[351,0],[362,0]]
       }
       middle_metacarpal_r_ctrl.Rotation.Z {type=MovieSceneScriptingFloatChannel, default=0, thinned_from=12, keys [[212,0],[362,0]]}
       middle_metacarpal_r_ctrl.Scale.X {type=MovieSceneScriptingFloatChannel, default=0, const [0,12]}
       middle_metacarpal_r_ctrl.Scale.Y {type=MovieSceneScriptingFloatChannel, default=0, const [0,12]}
       middle_metacarpal_r_ctrl.Scale.Z {type=MovieSceneScriptingFloatChannel, default=0, const [0,12]}
       middle_01_r_ctrl.Location.X {type=MovieSceneScriptingFloatChannel, default=0, thinned_from=12, keys [[212,0],[336,0],[341,0.0002],[362,0]]}
       middle_01_r_ctrl.Location.Y {type=MovieSceneScriptingFloatChannel, default=0, thinned_from=12, keys [[212,0],[362,0]]}
       middle_01_r_ctrl.Location.Z {type=MovieSceneScriptingFloatChannel, default=0, thinned_from=12, keys [[212,0],[362,0]]}
       middle_01_r_ctrl.Rotation.X {
        type=MovieSceneScriptingFloatChannel
        default=0
        thinned_from=12
        keys [[212,0],[335,0],[336,0.459],[341,6.1967],[349,0.0648],[351,0],[362,0]]
       }
       middle_01_r_ctrl.Rotation.Y {
        type=MovieSceneScriptingFloatChannel
        default=0
        thinned_from=12
        keys [[212,0],[335,0],[336,-0.4275],[341,-5.7719],[349,-0.0603],[351,0],[362,0]]
       }
       middle_01_r_ctrl.Rotation.Z {
        type=MovieSceneScriptingFloatChannel
        default=0
        thinned_from=12
        keys [[212,0],[335,0],[336,1.269],[341,17.1321],[349,0.1791],[351,0],[362,0]]
       }
       middle_01_r_ctrl.Scale.X {type=MovieSceneScriptingFloatChannel, default=0, const [0,12]}
       middle_01_r_ctrl.Scale.Y {type=MovieSceneScriptingFloatChannel, default=0, const [0,12]}
       middle_01_r_ctrl.Scale.Z {type=MovieSceneScriptingFloatChannel, default=0, const [0,12]}
       middle_02_r_ctrl.Location.X {type=MovieSceneScriptingFloatChannel, default=0, thinned_from=12, keys [[212,0],[362,0]]}
       middle_02_r_ctrl.Location.Y {type=MovieSceneScriptingFloatChannel, default=0, thinned_from=12, keys [[212,0],[362,0]]}
       middle_02_r_ctrl.Location.Z {type=MovieSceneScriptingFloatChannel, default=0, thinned_from=12, keys [[212,0],[362,0]]}
       middle_02_r_ctrl.Rotation.X {
        type=MovieSceneScriptingFloatChannel
        default=0
        thinned_from=12
        keys [[212,0],[335,0],[336,0.0566],[341,0.7643],[349,0.008],[351,0],[362,0]]
       }
       middle_02_r_ctrl.Rotation.Y {
        type=MovieSceneScriptingFloatChannel
        default=0
        thinned_from=12
        keys [[212,0],[335,0],[336,0.0651],[341,0.8782],[349,0.0092],[351,0],[362,0]]
       }
       middle_02_r_ctrl.Rotation.Z {
        type=MovieSceneScriptingFloatChannel
        default=0
        thinned_from=12
        keys [[212,0],[335,0],[336,1.183],[341,15.9707],[349,0.1669],[351,0],[362,0]]
       }
       middle_02_r_ctrl.Scale.X {type=MovieSceneScriptingFloatChannel, default=0, const [0,12]}
       middle_02_r_ctrl.Scale.Y {type=MovieSceneScriptingFloatChannel, default=0, const [0,12]}
       middle_02_r_ctrl.Scale.Z {type=MovieSceneScriptingFloatChannel, default=0, const [0,12]}
       middle_03_r_ctrl.Location.X {type=MovieSceneScriptingFloatChannel, default=0, thinned_from=12, keys [[212,0],[336,0],[341,-0.0002],[362,0]]}
       middle_03_r_ctrl.Location.Y {type=MovieSceneScriptingFloatChannel, default=0, thinned_from=12, keys [[212,0],[336,0],[341,-0.0001],[362,0]]}
       middle_03_r_ctrl.Location.Z {type=MovieSceneScriptingFloatChannel, default=0, thinned_from=12, keys [[212,0],[362,0]]}
       middle_03_r_ctrl.Rotation.X {
        type=MovieSceneScriptingFloatChannel
        default=0
        thinned_from=12
        keys [[212,0],[335,0],[336,0.0415],[341,0.5608],[349,0.0059],[351,0],[362,0]]
       }
       middle_03_r_ctrl.Rotation.Y {
        type=MovieSceneScriptingFloatChannel
        default=0
        thinned_from=12
        keys [[212,0],[335,0],[336,-0.0181],[341,-0.244],[349,-0.0026],[351,0],[362,0]]
       }
       middle_03_r_ctrl.Rotation.Z {
        type=MovieSceneScriptingFloatChannel
        default=0
        thinned_from=12
        keys [[212,0],[335,0],[336,1.1142],[341,15.0417],[349,0.1572],[351,0],[362,0]]
       }
       middle_03_r_ctrl.Scale.X {type=MovieSceneScriptingFloatChannel, default=0, const [0,12]}
       middle_03_r_ctrl.Scale.Y {type=MovieSceneScriptingFloatChannel, default=0, const [0,12]}
       middle_03_r_ctrl.Scale.Z {type=MovieSceneScriptingFloatChannel, default=0, const [0,12]}
       ring_metacarpal_r_ctrl.Location.X {type=MovieSceneScriptingFloatChannel, default=0, thinned_from=12, keys [[212,0],[362,0]]}
       ring_metacarpal_r_ctrl.Location.Y {type=MovieSceneScriptingFloatChannel, default=0, thinned_from=12, keys [[212,0],[362,0]]}
       ring_metacarpal_r_ctrl.Location.Z {type=MovieSceneScriptingFloatChannel, default=0, thinned_from=12, keys [[212,0],[362,0]]}
       ring_metacarpal_r_ctrl.Rotation.X {type=MovieSceneScriptingFloatChannel, default=0, thinned_from=12, keys [[212,0],[362,0]]}
       ring_metacarpal_r_ctrl.Rotation.Y {
        type=MovieSceneScriptingFloatChannel
        default=0
        thinned_from=12
        keys [[212,0],[335,0],[336,0.0925],[341,1.2486],[349,0.013],[351,0],[362,0]]
       }
       ring_metacarpal_r_ctrl.Rotation.Z {
        type=MovieSceneScriptingFloatChannel
        default=0
        thinned_from=12
        keys [[212,0],[335,0],[336,-0.077],[341,-1.0399],[349,-0.0109],[351,0],[362,0]]
       }
       ring_metacarpal_r_ctrl.Scale.X {type=MovieSceneScriptingFloatChannel, default=0, const [0,12]}
       ring_metacarpal_r_ctrl.Scale.Y {type=MovieSceneScriptingFloatChannel, default=0, const [0,12]}
       ring_metacarpal_r_ctrl.Scale.Z {type=MovieSceneScriptingFloatChannel, default=0, const [0,12]}
       ring_01_r_ctrl.Location.X {type=MovieSceneScriptingFloatChannel, default=0, thinned_from=12, keys [[212,0],[336,0],[341,0.0002],[362,0]]}
       ring_01_r_ctrl.Location.Y {type=MovieSceneScriptingFloatChannel, default=0, thinned_from=12, keys [[212,0],[362,0]]}
       ring_01_r_ctrl.Location.Z {type=MovieSceneScriptingFloatChannel, default=0, thinned_from=12, keys [[212,0],[362,0]]}
       ring_01_r_ctrl.Rotation.X {
        type=MovieSceneScriptingFloatChannel
        default=0
        thinned_from=12
        keys [[212,0],[335,0],[336,0.1789],[341,2.4156],[349,0.0252],[351,0],[362,0]]
       }
       ring_01_r_ctrl.Rotation.Y {
        type=MovieSceneScriptingFloatChannel
        default=0
        thinned_from=12
        keys [[212,0],[335,0],[336,0.0203],[341,0.2737],[349,0.0029],[351,0],[362,0]]
       }
       ring_01_r_ctrl.Rotation.Z {
        type=MovieSceneScriptingFloatChannel
        default=0
        thinned_from=12
        keys [[212,0],[335,0],[336,1.2322],[341,16.6351],[349,0.1739],[351,0],[362,0]]
       }
       ring_01_r_ctrl.Scale.X {type=MovieSceneScriptingFloatChannel, default=0, const [0,12]}
       ring_01_r_ctrl.Scale.Y {type=MovieSceneScriptingFloatChannel, default=0, const [0,12]}
       ring_01_r_ctrl.Scale.Z {type=MovieSceneScriptingFloatChannel, default=0, const [0,12]}
       ring_02_r_ctrl.Location.X {type=MovieSceneScriptingFloatChannel, default=0, thinned_from=12, keys [[212,0],[362,0]]}
       ring_02_r_ctrl.Location.Y {type=MovieSceneScriptingFloatChannel, default=0, thinned_from=12, keys [[212,0],[336,0],[341,0.0002],[362,0]]}
       ring_02_r_ctrl.Location.Z {type=MovieSceneScriptingFloatChannel, default=0, thinned_from=12, keys [[212,0],[362,0]]}
       ring_02_r_ctrl.Rotation.X {
        type=MovieSceneScriptingFloatChannel
        default=0
        thinned_from=12
        keys [[212,0],[335,0],[336,-0.0039],[341,-0.0526],[349,-0.0005],[351,0],[362,0]]
       }
       ring_02_r_ctrl.Rotation.Y {type=MovieSceneScriptingFloatChannel, default=0, thinned_from=12, keys [[212,0],[335,0],[336,-0.0004],[341,-0.0049],[349,-0.0001],[362,0]]}
       ring_02_r_ctrl.Rotation.Z {
        type=MovieSceneScriptingFloatChannel
        default=0
        thinned_from=12
        keys [[212,0],[335,0],[336,1.7365],[341,23.4421],[349,0.245],[351,0],[362,0]]
       }
       ring_02_r_ctrl.Scale.X {type=MovieSceneScriptingFloatChannel, default=0, const [0,12]}
       ring_02_r_ctrl.Scale.Y {type=MovieSceneScriptingFloatChannel, default=0, const [0,12]}
       ring_02_r_ctrl.Scale.Z {type=MovieSceneScriptingFloatChannel, default=0, const [0,12]}
       ring_03_r_ctrl.Location.X {type=MovieSceneScriptingFloatChannel, default=0, thinned_from=12, keys [[212,0],[336,0],[341,0.0001],[362,0]]}
       ring_03_r_ctrl.Location.Y {type=MovieSceneScriptingFloatChannel, default=0, thinned_from=12, keys [[212,0],[362,0]]}
       ring_03_r_ctrl.Location.Z {type=MovieSceneScriptingFloatChannel, default=0, thinned_from=12, keys [[212,0],[362,0]]}
       ring_03_r_ctrl.Rotation.X {type=MovieSceneScriptingFloatChannel, default=0, thinned_from=12, keys [[212,0],[335,0],[336,0.0008],[341,0.0105],[349,0.0001],[362,0]]}
       ring_03_r_ctrl.Rotation.Y {type=MovieSceneScriptingFloatChannel, default=0, thinned_from=12, keys [[212,0],[335,0],[336,-0.0007],[341,-0.0095],[349,-0.0001],[362,0]]}
       ring_03_r_ctrl.Rotation.Z {
        type=MovieSceneScriptingFloatChannel
        default=0
        thinned_from=12
        keys [[212,0],[335,0],[336,0.1374],[341,1.8548],[349,0.0194],[351,0],[362,0]]
       }
       ring_03_r_ctrl.Scale.X {type=MovieSceneScriptingFloatChannel, default=0, const [0,12]}
       ring_03_r_ctrl.Scale.Y {type=MovieSceneScriptingFloatChannel, default=0, const [0,12]}
       ring_03_r_ctrl.Scale.Z {type=MovieSceneScriptingFloatChannel, default=0, const [0,12]}
       pinky_metacarpal_r_ctrl.Location.X {type=MovieSceneScriptingFloatChannel, default=0, thinned_from=12, keys [[212,0],[362,0]]}
       pinky_metacarpal_r_ctrl.Location.Y {type=MovieSceneScriptingFloatChannel, default=0, thinned_from=12, keys [[212,0],[341,-0.0001],[362,0]]}
       pinky_metacarpal_r_ctrl.Location.Z {type=MovieSceneScriptingFloatChannel, default=0, thinned_from=12, keys [[212,0],[362,0]]}
       pinky_metacarpal_r_ctrl.Rotation.X {
        type=MovieSceneScriptingFloatChannel
        default=0
        thinned_from=12
        keys [[212,0],[335,0],[336,-0.0026],[341,-0.0345],[349,-0.0004],[351,0],[362,0]]
       }
       pinky_metacarpal_r_ctrl.Rotation.Y {
        type=MovieSceneScriptingFloatChannel
        default=0
        thinned_from=12
        keys [[212,0],[335,0],[336,0.0402],[341,0.5427],[349,0.0057],[351,0],[362,0]]
       }
       pinky_metacarpal_r_ctrl.Rotation.Z {
        type=MovieSceneScriptingFloatChannel
        default=0
        thinned_from=12
        keys [[212,0],[335,0],[336,-0.1718],[341,-2.3196],[349,-0.0242],[351,0],[362,0]]
       }
       pinky_metacarpal_r_ctrl.Scale.X {type=MovieSceneScriptingFloatChannel, default=0, const [0,12]}
       pinky_metacarpal_r_ctrl.Scale.Y {type=MovieSceneScriptingFloatChannel, default=0, const [0,12]}
       pinky_metacarpal_r_ctrl.Scale.Z {type=MovieSceneScriptingFloatChannel, default=0, const [0,12]}
       pinky_01_r_ctrl.Location.X {type=MovieSceneScriptingFloatChannel, default=0, thinned_from=12, keys [[212,0],[336,0],[341,0.0002],[362,0]]}
       pinky_01_r_ctrl.Location.Y {type=MovieSceneScriptingFloatChannel, default=0, thinned_from=12, keys [[212,0],[362,0]]}
       pinky_01_r_ctrl.Location.Z {type=MovieSceneScriptingFloatChannel, default=0, thinned_from=12, keys [[212,0],[362,0]]}
       pinky_01_r_ctrl.Rotation.X {
        type=MovieSceneScriptingFloatChannel
        default=0
        thinned_from=12
        keys [[212,0],[335,0],[336,-0.082],[341,-1.107],[349,-0.0116],[351,0],[362,0]]
       }
       pinky_01_r_ctrl.Rotation.Y {
        type=MovieSceneScriptingFloatChannel
        default=0
        thinned_from=12
        keys [[212,0],[335,0],[336,0.7267],[341,9.811],[349,0.1025],[351,0],[362,0]]
       }
       pinky_01_r_ctrl.Rotation.Z {
        type=MovieSceneScriptingFloatChannel
        default=0
        thinned_from=12
        keys [[212,0],[335,0],[336,0.5084],[341,6.8636],[349,0.0717],[351,0],[362,0]]
       }
       pinky_01_r_ctrl.Scale.X {type=MovieSceneScriptingFloatChannel, default=0, const [0,12]}
       pinky_01_r_ctrl.Scale.Y {type=MovieSceneScriptingFloatChannel, default=0, const [0,12]}
       pinky_01_r_ctrl.Scale.Z {type=MovieSceneScriptingFloatChannel, default=0, const [0,12]}
       pinky_02_r_ctrl.Location.X {type=MovieSceneScriptingFloatChannel, default=0, thinned_from=12, keys [[212,0],[362,0]]}
       pinky_02_r_ctrl.Location.Y {type=MovieSceneScriptingFloatChannel, default=0, thinned_from=12, keys [[212,0],[362,0]]}
       pinky_02_r_ctrl.Location.Z {type=MovieSceneScriptingFloatChannel, default=0, thinned_from=12, keys [[212,0],[362,0]]}
       pinky_02_r_ctrl.Rotation.X {
        type=MovieSceneScriptingFloatChannel
        default=0
        thinned_from=12
        keys [[212,0],[335,0],[336,0.0094],[341,0.1266],[349,0.0013],[351,0],[362,0]]
       }
       pinky_02_r_ctrl.Rotation.Y {
        type=MovieSceneScriptingFloatChannel
        default=0
        thinned_from=12
        keys [[212,0],[335,0],[336,-0.0014],[341,-0.0189],[349,-0.0002],[351,0],[362,0]]
       }
       pinky_02_r_ctrl.Rotation.Z {
        type=MovieSceneScriptingFloatChannel
        default=0
        thinned_from=12
        keys [[212,0],[335,0],[336,1.7445],[341,23.5514],[349,0.2461],[351,0],[362,0]]
       }
       pinky_02_r_ctrl.Scale.X {type=MovieSceneScriptingFloatChannel, default=0, const [0,12]}
       pinky_02_r_ctrl.Scale.Y {type=MovieSceneScriptingFloatChannel, default=0, const [0,12]}
       pinky_02_r_ctrl.Scale.Z {type=MovieSceneScriptingFloatChannel, default=0, const [0,12]}
       pinky_03_r_ctrl.Location.X {type=MovieSceneScriptingFloatChannel, default=0, thinned_from=12, keys [[212,0],[362,0]]}
       pinky_03_r_ctrl.Location.Y {type=MovieSceneScriptingFloatChannel, default=0, thinned_from=12, keys [[212,0],[362,0]]}
       pinky_03_r_ctrl.Location.Z {type=MovieSceneScriptingFloatChannel, default=0, thinned_from=12, keys [[212,0],[362,0]]}
       pinky_03_r_ctrl.Rotation.X {
        type=MovieSceneScriptingFloatChannel
        default=0
        thinned_from=12
        keys [[212,0],[335,0],[336,0.0117],[341,0.1578],[349,0.0016],[351,0],[362,0]]
       }
       pinky_03_r_ctrl.Rotation.Y {type=MovieSceneScriptingFloatChannel, default=0, thinned_from=12, keys [[212,0],[335,0],[336,0.0007],[341,0.0088],[349,0.0001],[362,0]]}
       pinky_03_r_ctrl.Rotation.Z {
        type=MovieSceneScriptingFloatChannel
        default=0
        thinned_from=12
        keys [[212,0],[335,0],[336,0.4088],[341,5.5181],[349,0.0577],[351,0],[362,0]]
       }
       pinky_03_r_ctrl.Scale.X {type=MovieSceneScriptingFloatChannel, default=0, const [0,12]}
       pinky_03_r_ctrl.Scale.Y {type=MovieSceneScriptingFloatChannel, default=0, const [0,12]}
       pinky_03_r_ctrl.Scale.Z {type=MovieSceneScriptingFloatChannel, default=0, const [0,12]}
       hand_r_ik_ctrl.Location.X {
        type=MovieSceneScriptingFloatChannel
        default=0
        thinned_from=11
        keys [[212,-0.0081],[324,-0.0081],[335,0.0216],[336,0.0224],[349,0.0288],[351,0.0288],[357,0.0303],[362,0.0246]]
       }
       hand_r_ik_ctrl.Location.Y {
        type=MovieSceneScriptingFloatChannel
        default=0
        thinned_from=11
        keys [[212,0.0054],[324,0.0054],[335,0.0154],[336,0.0136],[349,-0.1052],[351,-0.1057],[357,-0.2672],[362,0.0198]]
       }
       hand_r_ik_ctrl.Location.Z {
        type=MovieSceneScriptingFloatChannel
        default=0
        thinned_from=11
        keys [[212,-0.0024],[324,-0.0024],[335,-0.1424],[336,-0.1408],[349,-0.0324],[351,-0.0323],[357,-0.035],[362,-0.0303]]
       }
       hand_r_ik_ctrl.Rotation.X {
        type=MovieSceneScriptingFloatChannel
        default=0
        thinned_from=11
        keys [[212,0],[324,0],[335,-1.1246],[336,-1.1069],[349,0.0953],[351,0.0982],[357,0.1547],[362,0]]
       }
       hand_r_ik_ctrl.Rotation.Y {
        type=MovieSceneScriptingFloatChannel
        default=0
        thinned_from=11
        keys [[212,0],[324,0],[335,-0.4511],[336,-0.4428],[349,0.1195],[351,0.1211],[357,0.1908],[362,0]]
       }
       hand_r_ik_ctrl.Rotation.Z {
        type=MovieSceneScriptingFloatChannel
        default=0
        thinned_from=11
        keys [[212,0],[324,0],[335,-0.0965],[336,-0.1044],[349,-0.1904],[351,-0.1907],[357,-0.4414],[362,0]]
       }
       hand_r_ik_ctrl.Scale.X {type=MovieSceneScriptingFloatChannel, default=0, const [0,11]}
       hand_r_ik_ctrl.Scale.Y {type=MovieSceneScriptingFloatChannel, default=0, const [0,11]}
       hand_r_ik_ctrl.Scale.Z {type=MovieSceneScriptingFloatChannel, default=0, const [0,11]}
       Softness_2 {type=MovieSceneScriptingFloatChannel, default=0, const [0,11]}
       hand_r_ikorient_ctrl.Rotation.Roll {type=MovieSceneScriptingFloatChannel, default=0, const [0,6]}
       hand_r_ikorient_ctrl.Rotation.Pitch {type=MovieSceneScriptingFloatChannel, default=0, const [0,6]}
       hand_r_ikorient_ctrl.Rotation.Yaw {type=MovieSceneScriptingFloatChannel, default=0, const [0,6]}
       arm_r_pv_ik_ctrl.Location.X {
        type=MovieSceneScriptingFloatChannel
        default=0
        thinned_from=11
        keys [[212,-0.0071],[324,-0.0071],[335,-0.0057],[336,-0.0052],[349,0.0155],[351,0.0156],[357,0.0389],[362,-0.0061]]
       }
       arm_r_pv_ik_ctrl.Location.Y {
        type=MovieSceneScriptingFloatChannel
        default=0
        thinned_from=11
        keys [[212,-0.0058],[324,-0.0058],[335,0.1381],[336,0.1359],[349,-0.0112],[351,-0.0114],[357,-0.0057],[362,-0.0167]]
       }
       arm_r_pv_ik_ctrl.Location.Z {
        type=MovieSceneScriptingFloatChannel
        default=0
        thinned_from=11
        keys [[212,0.004],[324,0.004],[335,-0.074],[336,-0.0737],[349,-0.0542],[351,-0.0541],[357,-0.0715],[362,-0.04]]
       }
       arm_r_pv_ik_ctrl.Rotation.X {
        type=MovieSceneScriptingFloatChannel
        default=0
        thinned_from=11
        keys [[212,0],[324,0],[335,-0.0536],[336,-0.0533],[349,-0.0361],[351,-0.036],[357,-0.0834],[362,0]]
       }
       arm_r_pv_ik_ctrl.Rotation.Y {
        type=MovieSceneScriptingFloatChannel
        default=0
        thinned_from=11
        keys [[212,0],[324,0],[335,-0.0893],[336,-0.0879],[349,0.0072],[351,0.0074],[357,0.0117],[362,0]]
       }
       arm_r_pv_ik_ctrl.Rotation.Z {
        type=MovieSceneScriptingFloatChannel
        default=0
        thinned_from=11
        keys [[212,0],[324,0],[335,0.0957],[336,0.094],[349,-0.0248],[351,-0.0252],[357,-0.0397],[362,0]]
       }
       arm_r_pv_ik_ctrl.Scale.X {type=MovieSceneScriptingFloatChannel, default=0, const [0,11]}
       arm_r_pv_ik_ctrl.Scale.Y {type=MovieSceneScriptingFloatChannel, default=0, const [0,11]}
       arm_r_pv_ik_ctrl.Scale.Z {type=MovieSceneScriptingFloatChannel, default=0, const [0,11]}
       body_ctrl.Location.X {type=MovieSceneScriptingFloatChannel, default=0, const [0,6]}
       body_ctrl.Location.Y {type=MovieSceneScriptingFloatChannel, default=0, const [0,6]}
       body_ctrl.Location.Z {type=MovieSceneScriptingFloatChannel, default=0, const [0,6]}
       body_ctrl.Rotation.X {type=MovieSceneScriptingFloatChannel, default=0, const [0,6]}
       body_ctrl.Rotation.Y {type=MovieSceneScriptingFloatChannel, default=0, const [0,6]}
       body_ctrl.Rotation.Z {type=MovieSceneScriptingFloatChannel, default=0, const [0,6]}
       body_ctrl.Scale.X {type=MovieSceneScriptingFloatChannel, default=0, const [0,6]}
       body_ctrl.Scale.Y {type=MovieSceneScriptingFloatChannel, default=0, const [0,6]}
       body_ctrl.Scale.Z {type=MovieSceneScriptingFloatChannel, default=0, const [0,6]}
       hips_ctrl.Location.X {type=MovieSceneScriptingFloatChannel, default=0, const [0,6]}
       hips_ctrl.Location.Y {type=MovieSceneScriptingFloatChannel, default=0, const [0,6]}
       hips_ctrl.Location.Z {type=MovieSceneScriptingFloatChannel, default=0, const [0,6]}
       hips_ctrl.Rotation.X {type=MovieSceneScriptingFloatChannel, default=0, const [0,6]}
       hips_ctrl.Rotation.Y {type=MovieSceneScriptingFloatChannel, default=0, const [0,6]}
       hips_ctrl.Rotation.Z {type=MovieSceneScriptingFloatChannel, default=0, const [0,6]}
       hips_ctrl.Scale.X {type=MovieSceneScriptingFloatChannel, default=0, const [0,6]}
       hips_ctrl.Scale.Y {type=MovieSceneScriptingFloatChannel, default=0, const [0,6]}
       hips_ctrl.Scale.Z {type=MovieSceneScriptingFloatChannel, default=0, const [0,6]}
       hips_tan_ctrl.Location.X {type=MovieSceneScriptingFloatChannel, default=0, thinned_from=6, keys [[324,-0.0003],[335,0],[349,-0.0064],[357,-0.0062],[362,-0.0087]]}
       hips_tan_ctrl.Location.Y {type=MovieSceneScriptingFloatChannel, default=0, keys [[324,0],[335,-0.0006],[349,-0.0361],[351,-0.0362],[357,-0.0293],[362,-0.0415]]}
       hips_tan_ctrl.Location.Z {type=MovieSceneScriptingFloatChannel, default=0, keys [[324,0],[335,0.0014],[349,-0.0556],[351,-0.0557],[357,-0.0375],[362,-0.0688]]}
       hips_tan_ctrl.Rotation.X {type=MovieSceneScriptingFloatChannel, default=0, const [0,6]}
       hips_tan_ctrl.Rotation.Y {type=MovieSceneScriptingFloatChannel, default=0, const [0,6]}
       hips_tan_ctrl.Rotation.Z {type=MovieSceneScriptingFloatChannel, default=0, const [0,6]}
       hips_tan_ctrl.Scale.X {type=MovieSceneScriptingFloatChannel, default=0, const [0,6]}
       hips_tan_ctrl.Scale.Y {type=MovieSceneScriptingFloatChannel, default=0, const [0,6]}
       hips_tan_ctrl.Scale.Z {type=MovieSceneScriptingFloatChannel, default=0, const [0,6]}
       spine_01_ctrl.Location.X {type=MovieSceneScriptingFloatChannel, default=0, thinned_from=6, keys [[324,0],[362,0]]}
       spine_01_ctrl.Location.Y {type=MovieSceneScriptingFloatChannel, default=0, thinned_from=6, keys [[324,0],[362,0]]}
       spine_01_ctrl.Location.Z {type=MovieSceneScriptingFloatChannel, default=0, thinned_from=6, keys [[324,0],[362,0]]}
       spine_01_ctrl.Rotation.X {type=MovieSceneScriptingFloatChannel, default=0, const [0,6]}
       spine_01_ctrl.Rotation.Y {type=MovieSceneScriptingFloatChannel, default=0, const [0,6]}
       spine_01_ctrl.Rotation.Z {type=MovieSceneScriptingFloatChannel, default=0, const [0,6]}
       spine_01_ctrl.Scale.X {type=MovieSceneScriptingFloatChannel, default=0, const [0,6]}
       spine_01_ctrl.Scale.Y {type=MovieSceneScriptingFloatChannel, default=0, const [0,6]}
       spine_01_ctrl.Scale.Z {type=MovieSceneScriptingFloatChannel, default=0, const [0,6]}
       spine_02_ctrl.Location.X {type=MovieSceneScriptingFloatChannel, default=0, const [0,6]}
       spine_02_ctrl.Location.Y {type=MovieSceneScriptingFloatChannel, default=0, const [0,6]}
       spine_02_ctrl.Location.Z {type=MovieSceneScriptingFloatChannel, default=0, const [0,6]}
       spine_02_ctrl.Rotation.X {type=MovieSceneScriptingFloatChannel, default=0, const [0,6]}
       spine_02_ctrl.Rotation.Y {type=MovieSceneScriptingFloatChannel, default=0, const [0,6]}
       spine_02_ctrl.Rotation.Z {type=MovieSceneScriptingFloatChannel, default=0, const [0,6]}
       spine_02_ctrl.Scale.X {type=MovieSceneScriptingFloatChannel, default=0, const [0,6]}
       spine_02_ctrl.Scale.Y {type=MovieSceneScriptingFloatChannel, default=0, const [0,6]}
       spine_02_ctrl.Scale.Z {type=MovieSceneScriptingFloatChannel, default=0, const [0,6]}
       spine_03_ctrl.Location.X {type=MovieSceneScriptingFloatChannel, default=0, const [0,6]}
       spine_03_ctrl.Location.Y {type=MovieSceneScriptingFloatChannel, default=0, const [0,6]}
       spine_03_ctrl.Location.Z {type=MovieSceneScriptingFloatChannel, default=0, const [0,6]}
       spine_03_ctrl.Rotation.X {type=MovieSceneScriptingFloatChannel, default=0, const [0,6]}
       spine_03_ctrl.Rotation.Y {type=MovieSceneScriptingFloatChannel, default=0, const [0,6]}
       spine_03_ctrl.Rotation.Z {type=MovieSceneScriptingFloatChannel, default=0, const [0,6]}
       spine_03_ctrl.Scale.X {type=MovieSceneScriptingFloatChannel, default=0, const [0,6]}
       spine_03_ctrl.Scale.Y {type=MovieSceneScriptingFloatChannel, default=0, const [0,6]}
       spine_03_ctrl.Scale.Z {type=MovieSceneScriptingFloatChannel, default=0, const [0,6]}
       spine_04_ctrl.Location.X {type=MovieSceneScriptingFloatChannel, default=0, thinned_from=6, keys [[324,0],[362,0]]}
       spine_04_ctrl.Location.Y {type=MovieSceneScriptingFloatChannel, default=0, thinned_from=6, keys [[324,0],[362,0]]}
       spine_04_ctrl.Location.Z {type=MovieSceneScriptingFloatChannel, default=0, thinned_from=6, keys [[324,0],[362,0]]}
       spine_04_ctrl.Rotation.X {type=MovieSceneScriptingFloatChannel, default=0, const [0,6]}
       spine_04_ctrl.Rotation.Y {type=MovieSceneScriptingFloatChannel, default=0, const [0,6]}
       spine_04_ctrl.Rotation.Z {type=MovieSceneScriptingFloatChannel, default=0, const [0,6]}
       spine_04_ctrl.Scale.X {type=MovieSceneScriptingFloatChannel, default=0, const [0,6]}
       spine_04_ctrl.Scale.Y {type=MovieSceneScriptingFloatChannel, default=0, const [0,6]}
       spine_04_ctrl.Scale.Z {type=MovieSceneScriptingFloatChannel, default=0, const [0,6]}
       head_ctrl.Location.X {type=MovieSceneScriptingFloatChannel, default=0, const [0,6]}
       head_ctrl.Location.Y {type=MovieSceneScriptingFloatChannel, default=0, const [0,6]}
       head_ctrl.Location.Z {type=MovieSceneScriptingFloatChannel, default=0, const [0,6]}
       head_ctrl.Rotation.X {type=MovieSceneScriptingFloatChannel, default=0, thinned_from=6, keys [[324,0],[335,0.5308],[349,0.6284],[357,0.6288],[362,0]]}
       head_ctrl.Rotation.Y {type=MovieSceneScriptingFloatChannel, default=0, thinned_from=6, keys [[324,0],[335,-5.9601],[349,-7.056],[357,-7.0602],[362,0]]}
       head_ctrl.Rotation.Z {type=MovieSceneScriptingFloatChannel, default=0, thinned_from=6, keys [[324,0],[335,15.466],[349,18.3097],[357,18.3207],[362,0]]}
       head_ctrl.Scale.X {type=MovieSceneScriptingFloatChannel, default=0, const [0,6]}
       head_ctrl.Scale.Y {type=MovieSceneScriptingFloatChannel, default=0, const [0,6]}
       head_ctrl.Scale.Z {type=MovieSceneScriptingFloatChannel, default=0, const [0,6]}
       clavicle_l_ctrl.Location.X {type=MovieSceneScriptingFloatChannel, default=0, const [0,6]}
       clavicle_l_ctrl.Location.Y {type=MovieSceneScriptingFloatChannel, default=0, const [0,6]}
       clavicle_l_ctrl.Location.Z {type=MovieSceneScriptingFloatChannel, default=0, const [0,6]}
       clavicle_l_ctrl.Rotation.X {type=MovieSceneScriptingFloatChannel, default=0, const [0,6]}
       clavicle_l_ctrl.Rotation.Y {type=MovieSceneScriptingFloatChannel, default=0, const [0,6]}
       clavicle_l_ctrl.Rotation.Z {type=MovieSceneScriptingFloatChannel, default=0, const [0,6]}
       clavicle_l_ctrl.Scale.X {type=MovieSceneScriptingFloatChannel, default=0, const [0,6]}
       clavicle_l_ctrl.Scale.Y {type=MovieSceneScriptingFloatChannel, default=0, const [0,6]}
       clavicle_l_ctrl.Scale.Z {type=MovieSceneScriptingFloatChannel, default=0, const [0,6]}
       clavicle_r_ctrl.Location.X {type=MovieSceneScriptingFloatChannel, default=0, const [0,11]}
       clavicle_r_ctrl.Location.Y {type=MovieSceneScriptingFloatChannel, default=0, const [0,11]}
       clavicle_r_ctrl.Location.Z {type=MovieSceneScriptingFloatChannel, default=0, const [0,11]}
       clavicle_r_ctrl.Rotation.X {type=MovieSceneScriptingFloatChannel, default=0, const [0,11]}
       clavicle_r_ctrl.Rotation.Y {type=MovieSceneScriptingFloatChannel, default=0, const [0,11]}
       clavicle_r_ctrl.Rotation.Z {type=MovieSceneScriptingFloatChannel, default=0, const [0,11]}
       clavicle_r_ctrl.Scale.X {type=MovieSceneScriptingFloatChannel, default=0, const [0,11]}
       clavicle_r_ctrl.Scale.Y {type=MovieSceneScriptingFloatChannel, default=0, const [0,11]}
       clavicle_r_ctrl.Scale.Z {type=MovieSceneScriptingFloatChannel, default=0, const [0,11]}
       neck_01_ctrl.Location.X {type=MovieSceneScriptingFloatChannel, default=0, const [0,7]}
       neck_01_ctrl.Location.Y {type=MovieSceneScriptingFloatChannel, default=0, const [0,7]}
       neck_01_ctrl.Location.Z {type=MovieSceneScriptingFloatChannel, default=0, const [0,7]}
       neck_01_ctrl.Rotation.X {type=MovieSceneScriptingFloatChannel, default=0, thinned_from=7, keys [[272,0],[324,0],[335,5.3396],[349,0.0056],[351,0],[362,0]]}
       neck_01_ctrl.Rotation.Y {type=MovieSceneScriptingFloatChannel, default=0, const [0,7]}
       neck_01_ctrl.Rotation.Z {type=MovieSceneScriptingFloatChannel, default=0, const [0,7]}
       neck_01_ctrl.Scale.X {type=MovieSceneScriptingFloatChannel, default=0, const [0,7]}
       neck_01_ctrl.Scale.Y {type=MovieSceneScriptingFloatChannel, default=0, const [0,7]}
       neck_01_ctrl.Scale.Z {type=MovieSceneScriptingFloatChannel, default=0, const [0,7]}
       neck_02_ctrl.Location.X {type=MovieSceneScriptingFloatChannel, default=0, const [0,7]}
       neck_02_ctrl.Location.Y {type=MovieSceneScriptingFloatChannel, default=0, thinned_from=7, keys [[272,0],[362,0]]}
       neck_02_ctrl.Location.Z {type=MovieSceneScriptingFloatChannel, default=0, thinned_from=7, keys [[272,0],[362,0]]}
       neck_02_ctrl.Rotation.X {type=MovieSceneScriptingFloatChannel, default=0, const [0,7]}
       neck_02_ctrl.Rotation.Y {type=MovieSceneScriptingFloatChannel, default=0, const [0,7]}
       neck_02_ctrl.Rotation.Z {type=MovieSceneScriptingFloatChannel, default=0, const [0,7]}
       neck_02_ctrl.Scale.X {type=MovieSceneScriptingFloatChannel, default=0, const [0,7]}
       neck_02_ctrl.Scale.Y {type=MovieSceneScriptingFloatChannel, default=0, const [0,7]}
       neck_02_ctrl.Scale.Z {type=MovieSceneScriptingFloatChannel, default=0, const [0,7]}
       head_ik_ctrl.Location.X {type=MovieSceneScriptingFloatChannel, default=0, keys [[324,-0.0009],[335,0.0019],[349,-0.0332],[351,-0.0333],[357,-0.0035],[362,-0.0497]]}
       head_ik_ctrl.Location.Y {type=MovieSceneScriptingFloatChannel, default=0, keys [[324,-0.0006],[335,-0.0039],[349,0.0169],[351,0.0169],[357,-0.0068],[362,-0.027]]}
       head_ik_ctrl.Location.Z {type=MovieSceneScriptingFloatChannel, default=0, keys [[324,0.0004],[335,-0.0049],[349,0.3281],[351,0.3284],[357,0.2192],[362,0.3884]]}
       head_ik_ctrl.Rotation.X {type=MovieSceneScriptingFloatChannel, default=0, const [0,6]}
       head_ik_ctrl.Rotation.Y {type=MovieSceneScriptingFloatChannel, default=0, const [0,6]}
       head_ik_ctrl.Rotation.Z {type=MovieSceneScriptingFloatChannel, default=0, const [0,6]}
       head_ik_ctrl.Scale.X {type=MovieSceneScriptingFloatChannel, default=0, const [0,6]}
       head_ik_ctrl.Scale.Y {type=MovieSceneScriptingFloatChannel, default=0, const [0,6]}
       head_ik_ctrl.Scale.Z {type=MovieSceneScriptingFloatChannel, default=0, const [0,6]}
       chest_ctrl.Location.X {type=MovieSceneScriptingFloatChannel, default=0, thinned_from=6, keys [[324,0.0012],[335,-0.0005],[351,0.0001],[357,0.0002],[362,0]]}
       chest_ctrl.Location.Y {type=MovieSceneScriptingFloatChannel, default=0, keys [[324,-0.0007],[335,0.0001],[349,0.0051],[351,0.0051],[357,0.0119],[362,0]]}
       chest_ctrl.Location.Z {type=MovieSceneScriptingFloatChannel, default=0, keys [[324,-0.0001],[335,0.0001],[349,-0.0421],[351,-0.0421],[357,-0.0397],[362,-0.0437]]}
       chest_ctrl.Rotation.X {type=MovieSceneScriptingFloatChannel, default=0, keys [[324,-0.004],[335,-0.0002],[349,0.0085],[351,0.0085],[357,0.0196],[362,0]]}
       chest_ctrl.Rotation.Y {type=MovieSceneScriptingFloatChannel, default=0, thinned_from=6, keys [[324,-0.0055],[335,-0.0008],[349,0.0004],[357,0.0008],[362,0]]}
       chest_ctrl.Rotation.Z {type=MovieSceneScriptingFloatChannel, default=0, keys [[324,0.0068],[335,-0.0012],[349,0.0043],[351,0.0043],[357,0.0094],[362,0]]}
       chest_ctrl.Scale.X {type=MovieSceneScriptingFloatChannel, default=0, const [0,6]}
       chest_ctrl.Scale.Y {type=MovieSceneScriptingFloatChannel, default=0, const [0,6]}
       chest_ctrl.Scale.Z {type=MovieSceneScriptingFloatChannel, default=0, const [0,6]}
       chest_tan_ctrl.Location.X {type=MovieSceneScriptingFloatChannel, default=0, thinned_from=6, keys [[324,0.0004],[357,-0.0004],[362,-0.0004]]}
       chest_tan_ctrl.Location.Y {type=MovieSceneScriptingFloatChannel, default=0, thinned_from=6, keys [[324,0],[335,0],[351,0.0005],[357,0.0004],[362,0.0022]]}
       chest_tan_ctrl.Location.Z {type=MovieSceneScriptingFloatChannel, default=0, keys [[324,0],[335,0],[349,0.0092],[351,0.0093],[357,0.0013],[362,0.0151]]}
       chest_tan_ctrl.Rotation.X {type=MovieSceneScriptingFloatChannel, default=0, keys [[324,0.0004],[335,0],[349,0.0112],[351,0.0112],[357,0.026],[362,0]]}
       chest_tan_ctrl.Rotation.Y {type=MovieSceneScriptingFloatChannel, default=0, thinned_from=6, keys [[324,0.0013],[335,0.0003],[357,-0.0004],[362,0]]}
       chest_tan_ctrl.Rotation.Z {type=MovieSceneScriptingFloatChannel, default=0, thinned_from=6, keys [[324,0.0036],[335,-0.0022],[349,-0.0002],[357,-0.0004],[362,0]]}
       chest_tan_ctrl.Scale.X {type=MovieSceneScriptingFloatChannel, default=0, const [0,6]}
       chest_tan_ctrl.Scale.Y {type=MovieSceneScriptingFloatChannel, default=0, const [0,6]}
       chest_tan_ctrl.Scale.Z {type=MovieSceneScriptingFloatChannel, default=0, const [0,6]}
      ]
     }
    ]
   }
   {type=MovieSceneSpawnTrack, sections [{type=MovieSceneSpawnSection, channels [None {type=MovieSceneScriptingBoolChannel, default=true}]}]}
   {
    type=MovieScene3DTransformTrack
    sections [{
      type=MovieScene3DTransformSection
      channels [
       Location.X {type=MovieSceneScriptingDoubleChannel, default=0, const [0,1]}
       Location.Y {type=MovieSceneScriptingDoubleChannel, default=0, const [0,1]}
       Location.Z {type=MovieSceneScriptingDoubleChannel, default=0, const [0,1]}
       Rotation.X {type=MovieSceneScriptingDoubleChannel, default=0, const [0,1]}
       Rotation.Y {type=MovieSceneScriptingDoubleChannel, default=0, const [0,1]}
       Rotation.Z {type=MovieSceneScriptingDoubleChannel, default=0, const [0,1]}
       Scale.X {type=MovieSceneScriptingDoubleChannel, default=1, const [1,1]}
       Scale.Y {type=MovieSceneScriptingDoubleChannel, default=1, const [1,1]}
       Scale.Z {type=MovieSceneScriptingDoubleChannel, default=1, const [1,1]}
      ]
     }]
   }
  ]
 }
]
display_rate {denominator=1, numerator=60}
tick_resolution {denominator=1, numerator=24000}
tracks [{type=MovieSceneCameraCutTrack, sections [{type=MovieSceneCameraCutSection, end_frame=368, start_frame=0}]}]
