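# Level Sequence that animates the PistolCanary Superhero cinematic NPC, including a facial performance.
# A cine camera with a spawn track holds a fixed transform, focal length, focus distance and aperture in the channels shown.
# The NPC's Fortnite_3L_Face_CtrlRig track carries dense, near per-frame keys on controls such as eye aim, starting before frame zero, which suggests baked or captured animation.
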LevelSequence /PistolCanary/Anim/Superman/PistolCanary_Superman_lhm1001
bindings [
 CineCameraActor {
  id=#1
  possessed_class=CineCameraActor
  tracks [
   {
    type=MovieScene3DTransformTrack
    sections [{
      type=MovieScene3DTransformSection
      channels [
       Location.X {type=MovieSceneScriptingDoubleChannel, default=30.7391}
       Location.Y {type=MovieSceneScriptingDoubleChannel, default=435.5093}
       Location.Z {type=MovieSceneScriptingDoubleChannel, default=47.5476}
       Rotation.X {type=MovieSceneScriptingDoubleChannel, default=0}
       Rotation.Y {type=MovieSceneScriptingDoubleChannel, default=5.0001}
       Rotation.Z {type=MovieSceneScriptingDoubleChannel, default=-93.8}
       Scale.X {type=MovieSceneScriptingDoubleChannel, default=1}
       Scale.Y {type=MovieSceneScriptingDoubleChannel, default=1}
       Scale.Z {type=MovieSceneScriptingDoubleChannel, default=1}
      ]
     }]
   }
   {type=MovieSceneSpawnTrack, sections [{type=MovieSceneSpawnSection, channels [None {type=MovieSceneScriptingBoolChannel, default=true}]}]}
  ]
 }
 CameraComponent {
  id=#2
  parent_id=#1
  possessed_class=CineCameraComponent
  tracks [
   {
    type=MovieSceneFloatTrack
    display_name=Current Focal Length
    sections [{type=MovieSceneFloatSection, channels [None {type=MovieSceneScriptingFloatChannel, default=50}]}]
   }
   {
    type=MovieSceneFloatTrack
    display_name="Manual Focus Distance (Focus Settings)"
    sections [{type=MovieSceneFloatSection, channels [None {type=MovieSceneScriptingFloatChannel, default=100000}]}]
   }
   {
    type=MovieSceneFloatTrack
    display_name=Current Aperture
    sections [{type=MovieSceneFloatSection, channels [None {type=MovieSceneScriptingFloatChannel, default=10}]}]
   }
  ]
 }
 PistolCanary_CinematicNPC_Superhero_Main {
  id=#3
  possessed_class=PistolCanary_CinematicNPC_Superhero_Main_C
  tracks [
   {
    type=MovieSceneControlRigParameterTrack
    display_name=Fortnite_3L_Face_CtrlRig
    sections [
     {
      type=MovieSceneControlRigParameterSection
      channels [
       CTRL_C_eyesAim.Location.X {
        type=MovieSceneScriptingFloatChannel
        default=0
        thinned_from=492
        keys [[-131,7.2541],[-130,6.7031],[-129,6.8052],[-128,6.7499],[-127,6.4115],[-126,6.0574],[-125,5.9028],[-124,5.6641],[-122,5.6949],[-121,5.6075],[-120,5.4964],[-117,5.4871],[-116,5.4449],[-115,5.3766],[-114,5.1946],[-113,4.998],[-112,4.7174],[-111,4.5183],[-110,4.4873],[-109,4.6637],[-108,4.7831],[-107,4.8337],[-106,4.8422],[-105,4.8894],[-104,4.6972],[-103,4.4805],[-102,4.347],[-101,4.1694],[-100,3.9643],[-99,3.7798],[-98,3.5582],[-97,3.3724],[-96,3.2417],[-95,3.1967],[-94,3.0192],[-93,3.0087],[-92,3.0188],[-91,3.1376],[-90,3.224],[-88,3.6566],[-87,3.9322],[-86,4.1538],[-84,4.5445],[-83,4.6522],[-82,4.7975],[-80,4.8483],[-79,4.8624],[-78,4.9855],[-77,5.0627],[-76,4.9804],[-75,5.0998],[-74,4.8535],[-73,4.6618],[-72,4.5044],[-71,4.3805],[-70,4.4832],[-69,4.4394],[-68,4.4557],[-67,4.509],[-66,4.5348],[-65,4.6081],[-64,4.6082],[-63,4.7418],[-62,4.7255],[-60,4.7432],[-59,4.7115],[-58,4.8032],[-57,4.8],[-56,4.7835],[-55,4.8028],[-54,4.7639],[-53,4.7451],[-52,4.7716],[-51,4.8154],[-50,4.9252],[-49,4.9368],[-47,5.4155],[-46,5.4508],[-45,5.545],[-44,5.5649],[-43,5.648],[-42,5.7725],[-41,5.8117],[-40,5.6994],[-39,5.5254],[-38,5.4701],[-37,5.3296],[-36,5.2277],[-35,5.161],[-34,5.1968],[-33,5.1451],[-32,5.1169],[-31,5.1333],[-30,5.1668],[-28,5.1889],[-27,5.2132],[-26,5.2048],[-25,5.0784],[-24,5.0047],[-23,3.0534],[-22,2.0951],[-21,3.8083],[-20,4.281],[-19,4.285],[-18,4.3756],[-17,4.5626],[-16,4.7395],[-15,4.7643],[-14,4.7049],[-13,4.7095],[-12,4.7479],[-11,4.6288],[-10,4.7218],[-9,4.8026],[-8,4.7608],[-7,4.8453],[-6,4.8718],[-5,5.14],[-4,5.1777],[-3,5.2548],[-2,5.294],[-1,5.2913],[0,5.168],[1,5.1811],[2,5.0933],[3,5.1401],[4,5.1382],[6,5.2551],[7,5.3039],[9,5.2981],[10,5.3732],[11,5.4266],[13,5.4655],[14,5.5448],[15,5.5648],[16,5.5515],[17,5.5559],[18,5.6257],[19,5.648],[20,5.7027],[21,5.6173],[22,5.7046],[24,5.7435],[25,5.7272],[26,5.8051],[27,5.8314],[29,5.9156],[30,6.0145],[31,6.1702],[32,6.2303],[33,6.2483],[34,6.2839],[35,6.3036],[36,6.4318],[37,6.3306],[38,6.2489],[39,6.2378],[40,6.321],[41,6.3487],[42,6.2863],[43,6.3452],[44,6.3913],[45,6.3814],[46,6.3061],[47,6.1931],[48,5.9246],[49,5.726],[50,5.4317],[51,4.8284],[52,4.432],[53,3.4161],[54,3.2935],[55,2.86],[56,2.4164],[57,2.1776],[58,2.1518],[59,2.1847],[60,2.1922],[61,2.1447],[62,2.1122],[63,2.0985],[64,1.881],[65,1.881],[66,1.7114],[67,1.6906],[68,1.7518],[69,1.7588],[70,1.973],[71,2.1142],[72,2.2806],[73,2.757],[74,3.0268],[75,3.2032],[76,3.504],[77,3.8184],[78,3.8285],[79,3.9906],[80,4.202],[81,4.3349],[82,4.3675],[83,4.4551],[84,4.5669],[85,4.6588],[86,4.6617],[87,4.7025],[88,4.8367],[89,4.9117],[90,4.8382],[91,4.855],[92,4.7433],[95,4.7408],[96,4.6285],[97,4.6117],[98,4.5444],[99,4.4624],[100,4.4928],[101,4.5623],[102,4.5844],[103,4.6196],[104,4.7282],[105,4.7345],[106,4.7621],[107,4.8171],[108,4.91],[109,4.9839],[110,5.1512],[111,5.1552],[114,5.3059],[115,5.5514],[116,5.599],[117,5.6288],[118,5.6997],[119,5.826],[120,5.9157],[121,6.1521],[122,6.3218],[123,6.6539],[124,6.6473],[125,6.7648],[126,6.7872],[127,6.7611],[128,6.7732],[129,6.7254],[130,6.7094],[131,6.7502],[132,6.6109],[133,6.3568],[134,6.3042],[135,6.4753],[136,7.1458],[137,8.3237],[138,9.9316],[139,11.1469],[140,12.0405],[141,12.3543],[142,12.7351],[143,12.669],[145,12.4468],[146,12.4809],[147,12.3544],[148,12.4419],[149,12.3551],[150,12.3993],[152,12.2971],[154,12.2531],[155,12.1427],[156,12.1768],[158,12.0909],[159,11.9375],[160,12.1313],[161,12.2034],[162,12.1995],[163,12.2352],[164,12.4238],[165,12.3974],[166,12.4942],[167,12.7237],[168,13.0061],[170,12.9207],[172,12.2431],[173,11.7743],[174,11.3662],[175,10.9833],[176,10.716],[177,10.6297],[178,10.385],[179,10.3525],[180,10.2517],[181,10.2079],[183,10.1737],[184,10.1054],[185,9.9753],[186,9.7616],[187,9.1823],[188,9.0785],[189,8.6474],[190,7.9823],[191,7.5121],[192,7.5058],[193,8.0264],[194,10.4415],[195,10.8891],[196,10.9674],[197,10.2776],[198,8.6222],[199,6.9312],[200,6.5395],[201,5.7384],[202,3.767],[203,4.6276],[204,4.2231],[205,3.5113],[206,3.0208],[207,2.7129],[208,2.4483],[209,2.2465],[210,2.8384],[211,3.0575],[212,2.9294],[213,2.6042],[214,2.3234],[215,2.144],[216,2.0213],[217,1.925],[218,1.9585],[219,1.9717],[220,1.958],[221,2.0729],[222,2.1345],[224,2.2089],[225,2.2743],[226,2.4105],[227,2.5724],[228,2.8432],[229,3.1983],[230,3.613],[231,4.2728],[232,4.8886],[233,5.9701],[234,6.2981],[235,5.6367],[236,5.276],[237,3.6869],[238,1.7984],[239,3.4035],[240,2.8794],[241,0.9164],[242,0.5303],[243,0.2877],[244,0.9458],[245,1.7821],[246,2.7939],[247,2.8642],[248,3.0663],[249,3.5724],[250,4.0698],[251,4.1651],[252,4.5772],[253,5.1032],[254,5.342],[255,5.6424],[256,5.7717],[257,4.5402],[258,-0.0866],[259,-0.7139],[260,-7.1081],[261,-10.8392],[262,-12.2684],[263,-12.5837],[264,-12.5589],[265,-12.3909],[266,-12.2494],[267,-7.4316],[268,-3.2752],[269,4.8772],[270,6.4066],[271,9.8256],[272,9.2375],[273,8.5469],[274,7.2785],[275,6.6573],[276,6.6972],[277,6.6888],[278,6.5201],[279,6.4823],[280,6.4283],[281,6.4671],[282,6.4786],[283,6.5225],[284,6.5238],[285,6.4003],[286,6.5196],[289,6.5492],[290,6.5399],[291,6.3603],[292,5.9095],[293,5.6279],[294,5.8026],[295,6.0807],[296,6.0706],[297,6.0162],[298,4.8374],[299,4.0089],[300,2.6848],[301,2.604],[302,1.957],[303,1.8489],[304,1.523],[305,1.3336],[306,1.1106],[307,0.9642],[308,0.8598],[309,0.7971],[310,0.7768],[311,0.7384],[312,0.7247],[313,0.7296],[314,0.7741],[315,0.7583],[316,0.9044],[317,0.9417],[318,1.0559],[319,1.0837],[320,1.137],[321,1.0287],[322,1.0611],[323,1.096],[324,1.0178],[325,1.0277],[327,1.1735],[328,1.1237],[329,1.2136],[330,1.313],[331,1.3532],[332,1.3489],[333,1.5276],[334,1.6161],[335,1.6876],[336,1.7297],[337,1.719],[338,1.7866],[339,1.761],[340,1.842],[341,1.834],[342,1.8623],[343,1.7808],[344,1.836],[345,1.7769],[346,1.8243],[347,1.8162],[348,1.8548],[349,1.9713],[350,2.0187],[351,2.0495],[352,2.0876],[353,2.1477],[354,2.2169],[355,2.2532],[356,2.2496],[357,2.2837],[358,2.3656],[359,2.4392],[360,2.4392]]
       }
       CTRL_C_eyesAim.Location.Y {
        type=MovieSceneScriptingFloatChannel
        default=0
        thinned_from=492
        keys [[-131,1.0278],[-130,0.872],[-129,0.9045],[-128,0.8889],[-127,0.8127],[-126,0.734],[-125,0.7097],[-124,0.661],[-123,0.6725],[-122,0.6793],[-121,0.6514],[-120,0.6356],[-119,0.6248],[-118,0.6073],[-117,0.6047],[-116,0.5863],[-115,0.564],[-114,0.5098],[-113,0.4599],[-112,0.3986],[-111,0.3551],[-110,0.3428],[-109,0.3755],[-108,0.4021],[-107,0.414],[-106,0.4165],[-105,0.44],[-103,0.372],[-102,0.3443],[-101,0.3158],[-99,0.2645],[-98,0.2369],[-97,0.2161],[-96,0.202],[-95,0.1975],[-94,0.1823],[-93,0.1814],[-92,0.187],[-91,0.2063],[-90,0.2232],[-88,0.2938],[-87,0.347],[-86,0.3948],[-85,0.4278],[-84,0.4523],[-83,0.4691],[-82,0.4869],[-81,0.4798],[-80,0.4781],[-79,0.4799],[-78,0.4958],[-77,0.4953],[-76,0.4788],[-75,0.4992],[-74,0.4499],[-73,0.4167],[-72,0.3921],[-71,0.3717],[-70,0.4115],[-69,0.4076],[-68,0.411],[-67,0.4222],[-66,0.4197],[-65,0.419],[-64,0.42],[-63,0.4399],[-62,0.4382],[-61,0.4413],[-60,0.4433],[-59,0.432],[-58,0.4466],[-56,0.4474],[-55,0.4544],[-54,0.4492],[-53,0.4507],[-52,0.4585],[-51,0.4716],[-50,0.5116],[-49,0.5161],[-48,0.5664],[-47,0.6124],[-46,0.6216],[-45,0.6465],[-44,0.652],[-43,0.6721],[-42,0.6942],[-41,0.7013],[-40,0.678],[-39,0.635],[-38,0.6203],[-37,0.5726],[-36,0.5487],[-35,0.5311],[-34,0.5273],[-33,0.5096],[-32,0.4984],[-31,0.5004],[-30,0.5052],[-29,0.5051],[-28,0.504],[-27,0.5096],[-26,0.5113],[-25,0.4888],[-24,0.4742],[-23,0.1518],[-22,-0.0138],[-21,0.1931],[-20,0.2965],[-19,0.3323],[-18,0.3571],[-17,0.3908],[-16,0.4285],[-15,0.4405],[-14,0.4388],[-13,0.438],[-12,0.4464],[-11,0.4277],[-10,0.4452],[-9,0.4705],[-8,0.4615],[-7,0.4751],[-6,0.4772],[-5,0.5484],[-4,0.556],[-3,0.5738],[-2,0.5804],[-1,0.5715],[0,0.542],[1,0.543],[2,0.522],[3,0.5258],[4,0.523],[5,0.5322],[7,0.5572],[8,0.5566],[9,0.5582],[10,0.5766],[11,0.5895],[12,0.5943],[13,0.5949],[14,0.6147],[15,0.6262],[16,0.6269],[17,0.6233],[18,0.6388],[19,0.6443],[20,0.6585],[21,0.6448],[22,0.6641],[24,0.6715],[25,0.6705],[26,0.6862],[27,0.6886],[28,0.6988],[29,0.7122],[30,0.7449],[31,0.7885],[33,0.8075],[34,0.8249],[35,0.8295],[36,0.859],[37,0.8316],[38,0.8133],[39,0.8105],[40,0.8365],[41,0.8456],[42,0.836],[43,0.8605],[44,0.8788],[46,0.8719],[47,0.8366],[48,0.7727],[49,0.7189],[50,0.6517],[51,0.5219],[52,0.4592],[53,0.3074],[54,0.2933],[55,0.2325],[56,0.1697],[57,0.1243],[58,0.1014],[59,0.1079],[60,0.111],[61,0.1185],[62,0.1178],[63,0.1161],[64,0.0912],[65,0.0912],[66,0.0841],[67,0.0777],[68,0.0822],[69,0.0861],[70,0.0995],[71,0.1053],[72,0.1185],[73,0.1606],[74,0.1854],[75,0.2173],[76,0.2534],[77,0.3037],[78,0.3052],[79,0.33],[80,0.3569],[81,0.3914],[82,0.3869],[83,0.4051],[84,0.431],[85,0.4509],[86,0.4538],[87,0.4628],[88,0.4915],[89,0.5168],[90,0.5019],[91,0.5221],[92,0.5112],[93,0.5137],[94,0.524],[95,0.5281],[96,0.5037],[97,0.4993],[98,0.4817],[99,0.4721],[100,0.4923],[101,0.5103],[102,0.5142],[103,0.5304],[104,0.5625],[105,0.5644],[107,0.6127],[108,0.6531],[109,0.6886],[110,0.7351],[111,0.7549],[112,0.7797],[113,0.8236],[114,0.8366],[115,0.9242],[116,0.9656],[117,0.9967],[118,1.0523],[119,1.1132],[120,1.17],[121,1.2483],[122,1.2948],[123,1.4138],[124,1.4284],[125,1.4593],[126,1.4768],[128,1.4731],[129,1.452],[130,1.4552],[131,1.4635],[132,1.417],[133,1.3672],[134,1.4009],[135,1.4809],[136,1.6965],[137,2.1751],[138,2.8772],[139,3.401],[140,3.8954],[141,4.0704],[142,4.3078],[143,4.2368],[145,4.0462],[146,4.0507],[147,3.9864],[148,4.0309],[149,3.9968],[150,4.037],[151,4.018],[152,3.9814],[154,3.958],[155,3.8557],[156,3.8621],[157,3.83],[158,3.7775],[159,3.6786],[160,3.7595],[161,3.7871],[162,3.7609],[163,3.785],[164,3.8936],[165,3.8694],[166,3.9041],[167,4.0283],[168,4.1754],[169,4.1496],[170,4.1041],[171,3.8764],[172,3.666],[173,3.3739],[174,3.1228],[175,2.9021],[176,2.7607],[177,2.7157],[178,2.5844],[179,2.5516],[180,2.505],[181,2.4773],[182,2.4611],[183,2.4278],[184,2.3789],[185,2.305],[186,2.178],[187,1.8777],[188,1.826],[189,1.5998],[190,1.2947],[191,1.0912],[192,1.2055],[193,1.6628],[194,3.2984],[195,4.2028],[196,5.7973],[197,6.0754],[198,4.8491],[199,3.4018],[200,3.0698],[201,2.5235],[202,1.9387],[203,1.626],[204,1.2369],[205,0.8588],[206,0.6399],[207,0.4881],[208,0.3478],[209,0.2643],[210,0.3041],[211,0.2908],[212,0.253],[213,0.1869],[214,0.1381],[215,0.1189],[216,0.0949],[217,0.0721],[218,0.0661],[219,0.0634],[220,0.0564],[221,0.0605],[222,0.061],[223,0.0715],[224,0.0722],[225,0.0869],[226,0.1117],[227,0.1476],[228,0.2061],[229,0.2691],[230,0.3544],[231,0.5198],[232,0.6961],[233,1.0419],[234,1.0688],[235,0.5714],[236,0.4212],[237,0.1309],[238,-0.0819],[239,0.1424],[240,0.591],[241,1.8695],[242,2.22],[243,2.178],[244,1.5462],[245,0.9896],[246,0.6253],[247,0.5267],[248,0.443],[249,0.3724],[250,0.3408],[251,0.2897],[252,0.3086],[253,0.3926],[254,0.4343],[255,0.5049],[256,0.5428],[257,0.2921],[258,-0.1158],[259,-0.1034],[260,0.9648],[261,2.4554],[262,3.2422],[263,3.5135],[264,3.4974],[265,3.3962],[266,3.3019],[267,1.1005],[268,0.1811],[269,0.4251],[270,0.7801],[271,2.0204],[272,1.7839],[273,1.5081],[274,1.0627],[275,0.8626],[276,0.851],[277,0.8323],[278,0.7841],[279,0.774],[280,0.7592],[281,0.769],[282,0.767],[283,0.7825],[284,0.7836],[285,0.753],[286,0.7819],[287,0.7831],[288,0.7872],[289,0.7949],[290,0.7936],[291,0.7355],[292,0.5701],[293,0.5032],[294,0.5444],[295,0.6164],[296,0.6081],[297,0.5942],[298,0.3388],[299,0.2036],[300,0.0514],[301,0.0418],[302,-0.0224],[303,-0.0298],[304,-0.0471],[305,-0.0491],[306,-0.0505],[309,-0.0501],[311,-0.0499],[312,-0.0486],[313,-0.0489],[314,-0.0521],[315,-0.0599],[316,-0.0639],[317,-0.0703],[318,-0.0656],[319,-0.0659],[320,-0.0624],[321,-0.0614],[322,-0.0671],[323,-0.0651],[324,-0.0624],[325,-0.0593],[326,-0.0569],[327,-0.052],[328,-0.053],[329,-0.0515],[330,-0.0481],[331,-0.0443],[332,-0.0447],[333,-0.0378],[334,-0.0365],[335,-0.0327],[336,-0.0293],[337,-0.0169],[338,-0.0094],[339,-0.0091],[340,-0.0032],[341,-0.0062],[342,-0.006],[343,-0.007],[344,-0.0055],[345,-0.0122],[346,-0.0113],[347,-0.0101],[348,-0.0105],[349,-0.02],[350,-0.018],[351,-0.0127],[352,-0.0059],[353,0.0002],[354,0.007],[355,0.0114],[356,0.0114],[357,0.0168],[358,0.0291],[359,0.0338],[360,0.0338]]
       }
       CTRL_C_eyesAim.Location.Z {
        type=MovieSceneScriptingFloatChannel
        default=0
        thinned_from=492
        keys [[-131,0.739],[-130,0.6464],[-129,0.6105],[-128,0.6068],[-127,0.414],[-126,0.2694],[-125,0.1318],[-124,0.0472],[-123,-0.0281],[-122,-0.0472],[-121,0.0135],[-120,-0.0735],[-119,0.0089],[-118,0.1678],[-117,0.181],[-116,0.2624],[-115,0.329],[-114,0.4698],[-113,0.5558],[-112,0.6053],[-111,0.665],[-110,0.7416],[-109,0.7512],[-108,0.7146],[-107,0.6941],[-106,0.6847],[-105,0.5284],[-104,0.4856],[-103,0.4042],[-102,0.4368],[-101,0.407],[-100,0.3171],[-99,0.2778],[-98,0.2193],[-97,0.1635],[-96,0.1282],[-95,0.1153],[-94,0.0458],[-93,0.0426],[-92,0.0041],[-91,-0.0326],[-90,-0.0811],[-89,-0.1234],[-88,-0.1561],[-87,-0.2364],[-86,-0.3197],[-85,-0.3057],[-84,-0.2132],[-83,-0.1816],[-82,-0.0919],[-81,0.0098],[-80,0.0738],[-79,0.0832],[-78,0.1692],[-77,0.3295],[-76,0.3217],[-75,0.3695],[-74,0.3494],[-73,0.302],[-72,0.2492],[-71,0.2254],[-70,0.0339],[-69,-0.0039],[-68,-0.0065],[-67,-0.0174],[-66,0.0472],[-65,0.1805],[-64,0.1717],[-63,0.2297],[-62,0.2152],[-61,0.2002],[-60,0.2002],[-59,0.2481],[-58,0.2825],[-57,0.272],[-56,0.2367],[-55,0.2073],[-54,0.1828],[-53,0.135],[-52,0.1116],[-51,0.0721],[-50,-0.0808],[-49,-0.0986],[-48,-0.0964],[-47,-0.0337],[-46,-0.0432],[-45,-0.0696],[-44,-0.0769],[-43,-0.0812],[-42,-0.0146],[-41,0.0067],[-40,-0.0272],[-38,0.0057],[-37,0.1471],[-36,0.1597],[-35,0.1878],[-34,0.2981],[-33,0.362],[-32,0.4144],[-31,0.4302],[-30,0.4556],[-29,0.4823],[-28,0.5188],[-27,0.5151],[-26,0.4779],[-25,0.4292],[-24,0.4191],[-23,0.379],[-22,1.172],[-21,1.1886],[-20,0.8484],[-19,0.4469],[-18,0.3594],[-17,0.3701],[-16,0.3362],[-15,0.2661],[-14,0.1717],[-13,0.1874],[-12,0.1794],[-11,0.1368],[-10,0.1425],[-9,0.0596],[-8,0.0655],[-7,0.0955],[-6,0.1248],[-5,-0.0082],[-4,-0.0032],[-3,-0.0118],[-2,0.0065],[-1,0.0792],[0,0.1019],[1,0.1187],[2,0.1377],[3,0.1955],[4,0.2177],[5,0.2423],[6,0.2532],[7,0.2366],[8,0.2309],[9,0.215],[10,0.2004],[11,0.1926],[12,0.1973],[13,0.2256],[14,0.2121],[15,0.1487],[16,0.1141],[17,0.1564],[18,0.1644],[19,0.1626],[20,0.152],[21,0.0927],[22,0.106],[23,0.1135],[24,0.1234],[25,0.0972],[26,0.1264],[27,0.1635],[28,0.1568],[29,0.1386],[30,0.069],[31,0.0385],[32,0.0879],[33,0.0517],[34,-0.0168],[35,-0.0112],[36,0.0333],[37,0.0324],[38,0.0026],[39,0.0013],[40,-0.0311],[41,-0.0452],[42,-0.1047],[43,-0.176],[44,-0.2219],[45,-0.22],[46,-0.3471],[47,-0.3066],[48,-0.3451],[49,-0.3078],[50,-0.3231],[51,-0.3246],[52,-0.432],[53,-0.5373],[54,-0.5586],[55,-0.5179],[56,-0.4091],[57,-0.2286],[58,-0.0559],[59,-0.0865],[60,-0.1067],[61,-0.2057],[62,-0.2227],[63,-0.219],[64,-0.1602],[65,-0.1602],[66,-0.205],[67,-0.1642],[68,-0.1655],[69,-0.1938],[70,-0.1686],[71,-0.1175],[72,-0.102],[73,-0.0382],[74,0.0267],[75,-0.0538],[76,-0.0049],[77,-0.0272],[78,-0.0264],[79,-0.0132],[80,0.0676],[81,-0.0283],[82,0.0632],[83,0.0443],[84,0.001],[85,-0.0178],[86,-0.0384],[87,-0.0466],[88,-0.0639],[89,-0.1476],[90,-0.148],[91,-0.2838],[92,-0.3736],[93,-0.4003],[94,-0.4735],[95,-0.5038],[96,-0.4861],[97,-0.478],[98,-0.4423],[99,-0.4863],[100,-0.5916],[101,-0.6257],[102,-0.6234],[103,-0.6898],[104,-0.7642],[105,-0.7682],[106,-0.8926],[107,-0.9771],[108,-1.1114],[109,-1.2335],[110,-1.2967],[111,-1.4055],[112,-1.4818],[113,-1.6563],[114,-1.6718],[115,-1.8337],[116,-1.9851],[117,-2.1022],[118,-2.2843],[119,-2.42],[120,-2.5717],[121,-2.6442],[122,-2.6477],[123,-2.7555],[124,-2.824],[125,-2.8097],[126,-2.8548],[127,-2.873],[128,-2.8563],[129,-2.8274],[130,-2.8597],[131,-2.8448],[132,-2.8204],[133,-2.9095],[134,-3.1024],[135,-3.2283],[136,-3.287],[137,-3.5531],[138,-3.661],[139,-3.3948],[140,-3.3785],[141,-3.3371],[142,-3.3407],[144,-3.1503],[145,-3.0613],[146,-3.0038],[147,-3.0488],[148,-3.0181],[149,-3.0837],[150,-3.1304],[151,-3.1844],[152,-3.1521],[153,-3.1606],[154,-3.1627],[155,-3.0342],[156,-2.9849],[157,-2.9455],[158,-2.8618],[159,-2.8256],[160,-2.7059],[161,-2.6495],[162,-2.5572],[163,-2.5687],[164,-2.5513],[165,-2.5193],[166,-2.4249],[167,-2.3557],[168,-2.2277],[169,-2.2093],[170,-2.1589],[171,-2.1134],[172,-2.0636],[173,-1.9649],[174,-1.8377],[175,-1.7403],[176,-1.7067],[177,-1.6945],[178,-1.6301],[179,-1.5426],[180,-1.5466],[181,-1.5098],[182,-1.4504],[183,-1.3325],[184,-1.2375],[185,-1.1554],[186,-0.9703],[187,-0.6219],[188,-0.557],[189,-0.1321],[190,0.435],[191,0.9631],[192,-0.1842],[193,-1.8914],[194,-4.2265],[195,-5.9961],[196,-8.9408],[197,-9.9625],[198,-9.3512],[199,-7.9286],[200,-7.5092],[201,-6.8142],[202,-6.4301],[203,-5.1808],[204,-4.2208],[205,-3.2858],[206,-2.6608],[207,-2.1283],[208,-1.5434],[209,-1.1685],[210,-1.0265],[211,-0.761],[212,-0.6073],[213,-0.3919],[214,-0.2313],[215,-0.209],[216,-0.0959],[217,0.0353],[218,0.114],[219,0.1493],[220,0.2045],[221,0.2571],[222,0.3054],[223,0.2355],[224,0.2594],[225,0.1767],[226,0.068],[227,-0.1012],[228,-0.332],[229,-0.4743],[230,-0.6682],[231,-1.0462],[232,-1.3928],[233,-1.9085],[234,-1.6009],[235,0.9203],[236,2.451],[237,2.1784],[238,2.1909],[239,4.4475],[240,8.0148],[241,12.4658],[242,13.2679],[243,13.1883],[244,11.6759],[246,8.212],[247,7.7241],[248,7.1709],[250,5.4858],[251,4.8091],[252,3.9963],[253,3.4623],[254,2.8773],[255,2.4367],[256,2.091],[257,1.6976],[258,1.3786],[259,1.341],[260,0.964],[261,0.7801],[262,0.5637],[263,-0.1405],[264,-0.1316],[265,-0.1232],[266,-0.0371],[267,0.5519],[268,0.3756],[269,0.6668],[270,0.7547],[271,0.3951],[272,0.2769],[273,0.3332],[274,0.443],[275,0.6058],[276,0.8746],[277,1.0934],[278,1.1073],[279,1.1032],[280,1.1038],[281,1.1153],[282,1.1914],[283,1.1408],[284,1.129],[285,1.0842],[286,1.1365],[287,1.1358],[288,1.1199],[289,1.0673],[290,1.0507],[291,1.1823],[292,2.3916],[293,3.1677],[294,3.1318],[295,3.2698],[296,2.8208],[297,2.7145],[298,1.9946],[299,1.5762],[300,1.0607],[301,1.072],[302,1.127],[303,1.1121],[304,1.0366],[306,0.7854],[307,0.7071],[308,0.6608],[309,0.6339],[310,0.6265],[311,0.6115],[312,0.5911],[313,0.5967],[314,0.6486],[315,0.7364],[316,0.8487],[317,0.9523],[318,0.9516],[319,0.9718],[320,0.9568],[321,0.879],[322,0.9742],[323,0.9678],[324,0.887],[325,0.8519],[326,0.8615],[327,0.8417],[328,0.8248],[329,0.8608],[330,0.8839],[331,0.8631],[332,0.8647],[333,0.9121],[334,0.9719],[335,0.9861],[336,0.98],[337,0.8066],[338,0.7733],[339,0.7472],[340,0.7482],[341,0.7775],[342,0.8013],[343,0.7383],[344,0.7701],[345,0.7991],[346,0.8321],[347,0.8098],[348,0.8519],[349,1.1087],[350,1.1377],[351,1.0967],[352,1.0446],[353,1.032],[354,1.0234],[355,1.0074],[356,1.0034],[357,0.9718],[358,0.9112],[359,0.9451],[360,0.9451]]
       }
       CTRL_C_eyesAim.Rotation.X {
        type=MovieSceneScriptingFloatChannel
        default=0
        thinned_from=492
        keys [[-131,1.732],[-130,1.5154],[-129,1.4313],[-128,1.4226],[-127,0.971],[-126,0.6321],[-125,0.3093],[-124,0.1108],[-123,-0.066],[-122,-0.1109],[-121,0.0316],[-120,-0.1726],[-119,0.0209],[-118,0.3938],[-117,0.4248],[-116,0.6157],[-115,0.7718],[-114,1.1018],[-113,1.3034],[-112,1.4193],[-111,1.559],[-110,1.7383],[-109,1.7609],[-108,1.6752],[-107,1.6272],[-106,1.605],[-105,1.2392],[-104,1.1389],[-103,0.9481],[-102,1.0246],[-101,0.9548],[-100,0.7441],[-99,0.6518],[-98,0.5145],[-97,0.3838],[-96,0.3009],[-95,0.2707],[-94,0.1075],[-93,0.1001],[-92,0.0096],[-91,-0.0767],[-90,-0.1903],[-89,-0.2897],[-88,-0.3666],[-87,-0.5553],[-86,-0.7512],[-85,-0.7183],[-84,-0.5009],[-83,-0.4266],[-82,-0.2158],[-81,0.0231],[-80,0.1732],[-79,0.1952],[-78,0.3971],[-77,0.7729],[-76,0.7547],[-75,0.8669],[-74,0.8197],[-73,0.7085],[-72,0.5847],[-71,0.5289],[-70,0.0797],[-69,-0.0093],[-68,-0.0152],[-67,-0.0408],[-66,0.1107],[-65,0.4236],[-64,0.4029],[-63,0.539],[-62,0.5049],[-61,0.4697],[-60,0.4699],[-59,0.5823],[-58,0.6628],[-57,0.6382],[-56,0.5555],[-55,0.4864],[-54,0.4289],[-53,0.3169],[-52,0.2619],[-51,0.1693],[-50,-0.1897],[-49,-0.2316],[-48,-0.2263],[-47,-0.0791],[-46,-0.1015],[-45,-0.1634],[-44,-0.1805],[-43,-0.1907],[-42,-0.0344],[-41,0.0157],[-40,-0.0638],[-39,-0.025],[-38,0.0133],[-37,0.3452],[-36,0.3747],[-35,0.4408],[-34,0.6994],[-33,0.8493],[-32,0.972],[-31,1.009],[-30,1.0685],[-29,1.1312],[-28,1.2167],[-27,1.2081],[-26,1.1208],[-25,1.0068],[-24,0.983],[-23,0.8892],[-22,2.7453],[-21,2.784],[-20,1.9884],[-19,1.0482],[-18,0.8431],[-17,0.8682],[-16,0.7887],[-15,0.6243],[-14,0.4029],[-13,0.4397],[-12,0.421],[-11,0.3211],[-10,0.3344],[-9,0.14],[-8,0.1539],[-7,0.2242],[-6,0.2929],[-5,-0.0193],[-4,-0.0076],[-3,-0.0278],[-2,0.0151],[-1,0.186],[0,0.2392],[1,0.2786],[2,0.3231],[3,0.4587],[4,0.511],[5,0.5686],[6,0.594],[7,0.5552],[8,0.5419],[9,0.5045],[10,0.4703],[11,0.452],[12,0.4629],[13,0.5293],[14,0.4978],[15,0.349],[16,0.2677],[17,0.367],[18,0.3859],[19,0.3815],[20,0.3568],[21,0.2176],[22,0.2488],[23,0.2663],[24,0.2896],[25,0.2281],[26,0.2968],[27,0.3836],[28,0.3681],[29,0.3252],[30,0.162],[31,0.0904],[32,0.2064],[33,0.1214],[34,-0.0394],[35,-0.0262],[36,0.0783],[37,0.0761],[38,0.006],[39,0.0031],[40,-0.073],[41,-0.106],[42,-0.246],[43,-0.4134],[44,-0.5213],[45,-0.5168],[46,-0.8158],[47,-0.7203],[48,-0.8108],[49,-0.7233],[50,-0.7593],[51,-0.7626],[52,-1.0153],[53,-1.2632],[54,-1.3133],[55,-1.2176],[56,-0.9614],[57,-0.5371],[58,-0.1313],[59,-0.203],[60,-0.2506],[61,-0.4831],[62,-0.5232],[63,-0.5145],[64,-0.3764],[65,-0.3764],[66,-0.4816],[67,-0.3856],[68,-0.3888],[69,-0.4552],[70,-0.396],[71,-0.2758],[72,-0.2396],[73,-0.0897],[74,0.0627],[75,-0.1263],[76,-0.0115],[77,-0.064],[78,-0.0621],[79,-0.031],[80,0.1586],[81,-0.0664],[82,0.1484],[83,0.104],[84,0.0023],[85,-0.0419],[86,-0.0903],[87,-0.1095],[88,-0.1501],[89,-0.3466],[90,-0.3475],[91,-0.6669],[92,-0.8779],[93,-0.9408],[94,-1.1129],[95,-1.1844],[96,-1.1426],[97,-1.1236],[98,-1.0396],[99,-1.1431],[100,-1.3912],[101,-1.4715],[102,-1.4661],[103,-1.6224],[104,-1.7979],[105,-1.8073],[106,-2.1007],[107,-2.3002],[108,-2.6173],[109,-2.906],[110,-3.0557],[111,-3.3133],[112,-3.494],[113,-3.9079],[114,-3.9447],[115,-4.3294],[116,-4.6895],[117,-4.9685],[118,-5.4028],[119,-5.7271],[120,-6.0902],[121,-6.2643],[122,-6.2728],[123,-6.5322],[124,-6.6966],[125,-6.6626],[126,-6.771],[127,-6.8147],[128,-6.7745],[129,-6.705],[130,-6.7826],[131,-6.7468],[132,-6.688],[133,-6.9014],[134,-7.3654],[135,-7.6696],[136,-7.8136],[137,-8.4642],[138,-8.7385],[139,-8.0994],[140,-8.0681],[141,-7.9702],[142,-7.9831],[144,-7.5158],[145,-7.2981],[146,-7.1582],[147,-7.2667],[148,-7.1929],[149,-7.3517],[150,-7.4658],[151,-7.5969],[152,-7.5177],[153,-7.5384],[154,-7.5432],[155,-7.2295],[156,-7.1099],[157,-7.0141],[158,-6.8105],[159,-6.7219],[160,-6.4335],[161,-6.2975],[162,-6.0746],[163,-6.1026],[164,-6.0617],[165,-5.9841],[166,-5.7572],[167,-5.5915],[168,-5.2848],[169,-5.2403],[170,-5.119],[172,-4.8882],[173,-4.6505],[174,-4.346],[175,-4.113],[176,-4.0324],[177,-4.0033],[178,-3.8498],[179,-3.6417],[180,-3.651],[181,-3.5635],[182,-3.4225],[183,-3.1428],[184,-2.9176],[185,-2.7231],[186,-2.2851],[187,-1.4629],[188,-1.3097],[189,-0.3104],[190,1.0203],[191,2.2565],[192,-0.4327],[193,-4.4688],[194,-10.133],[195,-14.6054],[196,-22.5718],[197,-25.429],[198,-23.4569],[199,-19.4791],[200,-18.3637],[201,-16.5479],[202,-15.5413],[203,-12.428],[204,-10.0684],[205,-7.801],[206,-6.3],[207,-5.0288],[208,-3.6393],[209,-2.7519],[210,-2.4166],[211,-1.7902],[212,-1.4281],[213,-0.9211],[214,-0.5433],[215,-0.4909],[216,-0.2252],[217,0.0829],[218,0.2676],[219,0.3505],[220,0.48],[221,0.6034],[222,0.7166],[223,0.5526],[224,0.6088],[225,0.4148],[226,0.1596],[227,-0.2376],[228,-0.7801],[229,-1.1149],[230,-1.5715],[231,-2.4633],[232,-3.2832],[233,-4.5075],[234,-3.7769],[235,2.1564],[236,5.7317],[237,5.0957],[238,5.1249],[239,10.3942],[240,18.8182],[241,29.7394],[242,31.7909],[243,31.5845],[244,27.7467],[245,23.453],[246,19.29],[247,18.1234],[248,16.8074],[250,12.8302],[251,11.242],[252,9.3398],[253,8.0922],[254,6.7265],[255,5.6982],[256,4.8914],[257,3.9731],[258,3.2281],[259,3.1402],[260,2.2586],[261,1.828],[262,1.3213],[263,-0.3299],[264,-0.309],[265,-0.2893],[266,-0.0871],[267,1.2941],[268,0.881],[269,1.5632],[270,1.7689],[271,0.9267],[272,0.6496],[273,0.7817],[274,1.0389],[275,1.4203],[276,2.0494],[277,2.5613],[278,2.5938],[279,2.5842],[280,2.5855],[281,2.6125],[282,2.7904],[283,2.6721],[284,2.6444],[285,2.5398],[286,2.662],[287,2.6604],[288,2.6231],[289,2.5001],[290,2.4613],[291,2.769],[292,5.5929],[293,7.4045],[294,7.3207],[295,7.6431],[296,6.5948],[297,6.3466],[298,4.6665],[299,3.6896],[300,2.4851],[301,2.5114],[302,2.6402],[303,2.6051],[304,2.4287],[306,1.8411],[307,1.6576],[308,1.5493],[309,1.4863],[310,1.469],[311,1.4338],[312,1.386],[313,1.3992],[314,1.5207],[315,1.7264],[316,1.9892],[317,2.2314],[318,2.2298],[319,2.277],[320,2.2419],[321,2.0599],[322,2.2828],[323,2.2679],[324,2.0788],[325,1.9965],[326,2.0192],[327,1.9728],[328,1.9332],[329,2.0174],[330,2.0714],[331,2.0227],[332,2.0265],[333,2.1374],[334,2.2773],[335,2.3105],[336,2.2963],[337,1.8906],[338,1.8125],[339,1.7515],[340,1.7539],[341,1.8225],[342,1.8783],[343,1.7308],[344,1.8052],[345,1.8731],[346,1.9502],[347,1.8981],[348,1.9966],[349,2.5973],[350,2.665],[351,2.5691],[352,2.4473],[353,2.4179],[354,2.3979],[355,2.3604],[356,2.3511],[357,2.2771],[358,2.1353],[359,2.2146],[360,2.2146]]
       }
       CTRL_C_eyesAim.Rotation.Y {
        type=MovieSceneScriptingFloatChannel
        default=0
        thinned_from=492
        keys [[-131,-0.5391],[-130,-0.4328],[-129,-0.4156],[-128,-0.4094],[-127,-0.2644],[-126,-0.162],[-125,-0.0771],[-124,-0.0264],[-123,0.0158],[-122,0.0266],[-121,-0.0075],[-120,0.0399],[-119,-0.0048],[-118,-0.091],[-117,-0.098],[-116,-0.1409],[-115,-0.1743],[-114,-0.24],[-113,-0.2727],[-112,-0.2796],[-111,-0.2937],[-110,-0.3251],[-109,-0.3428],[-108,-0.3348],[-107,-0.3288],[-106,-0.3249],[-105,-0.2534],[-104,-0.2234],[-103,-0.1771],[-102,-0.1855],[-101,-0.1656],[-100,-0.1225],[-99,-0.1022],[-98,-0.0758],[-97,-0.0535],[-96,-0.0403],[-95,-0.0358],[-94,-0.0134],[-93,-0.0124],[-91,0.0099],[-90,0.0254],[-89,0.0413],[-88,0.0556],[-87,0.0907],[-86,0.1297],[-85,0.1301],[-84,0.0949],[-83,0.0828],[-82,0.0433],[-81,-0.0046],[-80,-0.0351],[-79,-0.0397],[-78,-0.0829],[-77,-0.1639],[-76,-0.1573],[-75,-0.1852],[-74,-0.1664],[-73,-0.1379],[-72,-0.1098],[-71,-0.0965],[-70,-0.0149],[-69,0.0017],[-68,0.0028],[-67,0.0077],[-66,-0.0209],[-65,-0.0815],[-64,-0.0775],[-63,-0.1068],[-62,-0.0997],[-61,-0.0929],[-60,-0.0931],[-59,-0.1146],[-58,-0.1331],[-57,-0.128],[-56,-0.111],[-55,-0.0976],[-54,-0.0854],[-53,-0.0628],[-52,-0.0522],[-51,-0.0341],[-50,0.0391],[-49,0.0478],[-48,0.0491],[-47,0.018],[-46,0.0233],[-45,0.0381],[-44,0.0423],[-43,0.0454],[-42,0.0084],[-41,-0.0039],[-40,0.0153],[-39,0.0058],[-38,-0.0031],[-37,-0.0772],[-36,-0.0822],[-35,-0.0954],[-34,-0.1524],[-33,-0.1832],[-32,-0.2084],[-31,-0.2171],[-29,-0.2456],[-28,-0.2647],[-27,-0.2641],[-26,-0.2446],[-25,-0.2142],[-24,-0.206],[-23,-0.1121],[-22,-0.2365],[-21,-0.4397],[-20,-0.3542],[-19,-0.187],[-18,-0.1537],[-17,-0.1652],[-16,-0.1561],[-15,-0.1243],[-14,-0.0792],[-13,-0.0865],[-12,-0.0835],[-11,-0.062],[-10,-0.066],[-9,-0.0281],[-8,-0.0306],[-7,-0.0454],[-6,-0.0597],[-5,0.0042],[-4,0.0016],[-3,0.0061],[-2,-0.0034],[-1,-0.0413],[0,-0.0518],[1,-0.0605],[2,-0.0689],[3,-0.0988],[4,-0.1101],[5,-0.1238],[6,-0.131],[7,-0.1236],[8,-0.1205],[9,-0.1122],[10,-0.1062],[11,-0.1031],[12,-0.1061],[13,-0.1216],[14,-0.1161],[15,-0.0817],[16,-0.0625],[17,-0.0858],[18,-0.0914],[19,-0.0908],[20,-0.0858],[21,-0.0515],[22,-0.0598],[23,-0.0642],[24,-0.0701],[25,-0.0551],[26,-0.0727],[27,-0.0944],[28,-0.0912],[29,-0.0813],[30,-0.0412],[31,-0.0236],[32,-0.0545],[33,-0.0322],[34,0.0105],[35,0.007],[36,-0.0214],[37,-0.0204],[38,-0.0016],[39,-0.0008],[40,0.0196],[41,0.0286],[42,0.0656],[43,0.1113],[44,0.1415],[45,0.14],[46,0.2182],[47,0.189],[48,0.2029],[49,0.1746],[50,0.1733],[51,0.1539],[52,0.1875],[53,0.1786],[54,0.1789],[55,0.1437],[56,0.0957],[57,0.0481],[58,0.0116],[59,0.0182],[60,0.0226],[61,0.0426],[62,0.0455],[63,0.0444],[64,0.0291],[65,0.0291],[66,0.0339],[67,0.0268],[68,0.028],[69,0.0329],[70,0.0321],[71,0.024],[72,0.0225],[73,0.0102],[74,-0.0078],[75,0.0167],[76,0.0017],[77,0.0101],[78,0.0099],[79,0.0051],[80,-0.0277],[81,0.012],[82,-0.027],[83,-0.0193],[84,-0.0004],[85,0.0081],[86,0.0176],[87,0.0215],[88,0.0304],[89,0.0712],[90,0.0703],[91,0.1354],[92,0.1739],[93,0.1862],[94,0.2204],[95,0.2345],[96,0.2207],[97,0.2162],[98,0.197],[99,0.2126],[100,0.2605],[101,0.28],[102,0.2804],[103,0.3127],[104,0.355],[105,0.3574],[106,0.4179],[107,0.463],[108,0.5374],[109,0.606],[110,0.6595],[111,0.7156],[112,0.7625],[113,0.862],[114,0.8779],[115,1.0103],[116,1.104],[117,1.1761],[118,1.2956],[119,1.4053],[120,1.5185],[121,1.6282],[122,1.6786],[123,1.8466],[124,1.8908],[125,1.9172],[126,1.9552],[127,1.9595],[128,1.9518],[129,1.9172],[130,1.9343],[131,1.9367],[132,1.8772],[133,1.8569],[134,1.9634],[135,2.1035],[136,2.384],[137,3.0571],[138,3.8723],[139,4.1377],[140,4.5518],[141,4.6534],[142,4.8563],[143,4.6788],[144,4.4878],[145,4.3076],[146,4.2414],[148,4.2439],[149,4.2961],[150,4.3831],[151,4.4318],[152,4.3647],[154,4.3584],[155,4.1289],[156,4.0763],[157,4.0062],[158,3.8696],[159,3.7559],[160,3.673],[161,3.6243],[162,3.4954],[163,3.525],[164,3.574],[165,3.5185],[166,3.4215],[167,3.407],[168,3.3202],[169,3.2808],[170,3.1873],[171,3.0044],[172,2.8292],[174,2.284],[175,2.0703],[176,1.9685],[177,1.9349],[178,1.8084],[179,1.7043],[180,1.6885],[181,1.6395],[182,1.5728],[183,1.4404],[184,1.3265],[185,1.219],[186,0.9969],[187,0.594],[188,0.5249],[189,0.1176],[190,-0.3531],[191,-0.7298],[192,0.1399],[193,1.5544],[194,4.7608],[195,7.1659],[196,10.9295],[197,11.2759],[198,8.5496],[199,5.6413],[200,5.008],[201,3.9417],[202,2.3971],[203,2.3801],[204,1.7594],[205,1.1306],[206,0.7842],[207,0.5618],[208,0.3667],[209,0.2543],[210,0.2829],[211,0.226],[212,0.1727],[213,0.0989],[214,0.052],[215,0.0433],[216,0.0187],[217,-0.0066],[218,-0.0215],[219,-0.0284],[220,-0.0386],[221,-0.0514],[222,-0.0629],[223,-0.0494],[224,-0.0553],[225,-0.0388],[226,-0.0158],[227,0.0252],[228,0.0915],[229,0.1474],[230,0.2352],[231,0.4379],[232,0.6709],[233,1.1359],[234,1.0081],[235,-0.5118],[236,-1.2669],[237,-0.7777],[238,-0.3782],[239,-1.4556],[240,-2.195],[241,-1.0679],[242,-0.6561],[243,-0.3538],[244,-1.0345],[245,-1.6692],[246,-2.1803],[247,-2.1054],[248,-2.0974],[249,-2.1666],[250,-2.151],[251,-1.9341],[252,-1.775],[253,-1.7242],[254,-1.5052],[255,-1.3517],[256,-1.189],[257,-0.7517],[258,0.0115],[259,0.0918],[260,0.6874],[261,0.906],[262,0.7683],[263,-0.1986],[264,-0.1855],[265,-0.1705],[266,-0.0505],[267,0.4137],[268,0.1193],[269,-0.3188],[270,-0.4812],[271,-0.4076],[272,-0.2657],[273,-0.2923],[274,-0.3246],[275,-0.4027],[276,-0.5847],[277,-0.7297],[278,-0.7189],[280,-0.7057],[281,-0.7177],[282,-0.768],[283,-0.7408],[284,-0.7333],[285,-0.69],[286,-0.7377],[287,-0.7377],[288,-0.7287],[289,-0.6963],[290,-0.6844],[291,-0.7472],[292,-1.3934],[293,-1.7495],[294,-1.7866],[295,-1.9599],[296,-1.6895],[297,-1.6107],[298,-0.9426],[299,-0.614],[300,-0.275],[301,-0.2694],[302,-0.2123],[303,-0.1979],[304,-0.1518],[305,-0.1169],[306,-0.0839],[307,-0.0655],[308,-0.0546],[309,-0.0486],[310,-0.0468],[311,-0.0434],[312,-0.0412],[313,-0.0418],[314,-0.0483],[315,-0.0537],[316,-0.0738],[317,-0.0861],[318,-0.0965],[319,-0.1012],[320,-0.1045],[321,-0.0869],[322,-0.0993],[323,-0.1019],[324,-0.0868],[325,-0.0841],[326,-0.091],[327,-0.095],[328,-0.0891],[330,-0.1116],[331,-0.1123],[332,-0.1122],[333,-0.134],[334,-0.1511],[335,-0.1601],[336,-0.1631],[337,-0.1335],[338,-0.133],[339,-0.1267],[340,-0.1327],[341,-0.1373],[342,-0.1437],[343,-0.1266],[344,-0.1362],[345,-0.1367],[346,-0.1462],[347,-0.1416],[348,-0.1522],[349,-0.2104],[350,-0.2211],[351,-0.2165],[352,-0.2101],[353,-0.2136],[354,-0.2187],[355,-0.2188],[356,-0.2176],[357,-0.214],[358,-0.2079],[359,-0.2224],[360,-0.2224]]
       }
       CTRL_C_eyesAim.Rotation.Z {
        type=MovieSceneScriptingFloatChannel
        default=0
        thinned_from=492
        keys [[-131,17.2845],[-130,15.9368],[-129,16.1869],[-128,16.0519],[-127,15.2296],[-126,14.3711],[-125,13.9974],[-124,13.4205],[-122,13.495],[-121,13.2841],[-120,13.0161],[-117,12.9935],[-116,12.8914],[-115,12.7265],[-114,12.2878],[-113,11.8149],[-112,11.1423],[-111,10.6657],[-110,10.5907],[-109,11.0121],[-108,11.2981],[-107,11.4196],[-106,11.4401],[-105,11.5549],[-104,11.0953],[-103,10.578],[-102,10.259],[-101,9.8359],[-100,9.3482],[-99,8.9096],[-98,8.3836],[-97,7.943],[-96,7.6333],[-95,7.5268],[-94,7.1068],[-93,7.0819],[-92,7.1057],[-91,7.3871],[-90,7.5914],[-88,8.6172],[-87,9.272],[-86,9.7992],[-84,10.7319],[-83,10.9896],[-82,11.3374],[-80,11.4591],[-79,11.4929],[-78,11.7877],[-77,11.9721],[-76,11.7748],[-75,12.061],[-74,11.4704],[-73,11.012],[-72,10.6361],[-71,10.3402],[-70,10.5859],[-69,10.4813],[-68,10.5202],[-67,10.6476],[-66,10.7092],[-65,10.8841],[-64,10.8843],[-63,11.2038],[-62,11.1648],[-60,11.2071],[-59,11.1311],[-58,11.3504],[-57,11.3429],[-56,11.3035],[-55,11.3499],[-54,11.2569],[-53,11.2119],[-52,11.2753],[-51,11.3803],[-50,11.6434],[-49,11.6712],[-47,12.8213],[-46,12.9064],[-45,13.1334],[-44,13.1814],[-43,13.3818],[-42,13.6824],[-41,13.7772],[-40,13.506],[-39,13.086],[-38,12.9528],[-37,12.6145],[-36,12.3694],[-35,12.2088],[-34,12.2944],[-33,12.1698],[-32,12.1016],[-31,12.1408],[-30,12.2211],[-28,12.2735],[-27,12.3319],[-26,12.3121],[-25,12.0091],[-24,11.8324],[-23,7.1868],[-22,4.9192],[-21,8.9672],[-20,10.0971],[-19,10.1112],[-18,10.328],[-17,10.7744],[-16,11.1976],[-15,11.2573],[-14,11.1157],[-13,11.1266],[-12,11.2184],[-11,10.9336],[-10,11.156],[-9,11.3497],[-8,11.2497],[-7,11.452],[-6,11.5154],[-5,12.1589],[-4,12.2495],[-3,12.4348],[-2,12.5289],[-1,12.5224],[0,12.226],[1,12.2574],[2,12.0466],[3,12.1588],[4,12.154],[6,12.4348],[7,12.5521],[9,12.5383],[10,12.7191],[11,12.8476],[13,12.9413],[14,13.1325],[15,13.1808],[16,13.149],[17,13.1593],[18,13.3276],[19,13.3815],[20,13.5136],[21,13.3076],[22,13.5184],[24,13.6123],[25,13.5728],[26,13.761],[27,13.8245],[29,14.0283],[30,14.2679],[31,14.6454],[32,14.7913],[33,14.835],[34,14.9216],[35,14.9694],[36,15.2813],[37,15.035],[38,14.8365],[39,14.8096],[40,15.0118],[41,15.0791],[42,14.9272],[43,15.0702],[44,15.1822],[45,15.158],[46,14.9741],[47,14.6998],[48,14.0488],[49,13.5689],[50,12.8592],[51,11.4105],[52,10.462],[53,8.0449],[54,7.7542],[55,6.7286],[56,5.6819],[57,5.1192],[58,5.0586],[59,5.1362],[60,5.1537],[61,5.0417],[62,4.9651],[63,4.9329],[64,4.4205],[65,4.4205],[66,4.0212],[67,3.9722],[68,4.1164],[69,4.1329],[70,4.6373],[71,4.9698],[72,5.3621],[73,6.4867],[74,7.1248],[75,7.5423],[76,8.2552],[77,9.0019],[78,9.0259],[79,9.4115],[80,9.9149],[81,10.232],[82,10.3097],[83,10.5187],[84,10.7858],[85,11.0056],[86,11.0124],[87,11.1101],[88,11.4314],[89,11.6108],[90,11.4347],[91,11.4744],[92,11.2065],[95,11.1994],[96,10.931],[97,10.8908],[98,10.7304],[99,10.5341],[100,10.6058],[101,10.7714],[102,10.8242],[103,10.9075],[104,11.1661],[105,11.1811],[106,11.2451],[107,11.3752],[108,11.5947],[109,11.7688],[110,12.1681],[111,12.1746],[114,12.5274],[115,13.1106],[116,13.2184],[117,13.2845],[118,13.4456],[119,13.7416],[120,13.948],[121,14.5126],[122,14.9218],[123,15.7171],[124,15.6958],[125,15.9818],[126,16.0323],[127,15.9675],[128,15.9984],[129,15.8847],[130,15.8435],[131,15.9434],[132,15.608],[133,14.9869],[134,14.8446],[135,15.2462],[136,16.8637],[137,19.7166],[138,23.7186],[139,26.8858],[140,29.2423],[141,30.0897],[142,31.1173],[143,30.9578],[145,30.3913],[146,30.4938],[147,30.1445],[148,30.386],[149,30.1402],[150,30.2506],[152,29.9715],[154,29.8515],[155,29.5786],[156,29.6787],[158,29.4693],[159,29.0652],[160,29.6023],[161,29.8046],[162,29.8081],[163,29.9024],[164,30.4141],[165,30.3475],[166,30.6234],[167,31.2585],[168,32.0519],[170,31.8262],[172,29.9911],[173,28.7436],[174,27.6725],[175,26.6735],[176,25.978],[177,25.7543],[178,25.1246],[179,25.0469],[180,24.7862],[181,24.6757],[183,24.5989],[184,24.4282],[185,24.0981],[186,23.5585],[187,22.0951],[188,21.8337],[189,20.7531],[190,19.0893],[191,17.914],[192,17.9122],[193,19.1415],[194,24.9098],[195,25.5809],[196,24.5186],[197,22.3551],[198,18.9119],[199,15.5312],[200,14.7335],[201,13.0266],[202,8.5527],[203,10.6718],[204,9.8104],[205,8.1956],[206,7.0673],[207,6.358],[208,5.7461],[209,5.2756],[210,6.6732],[211,7.1938],[212,6.8921],[213,6.1252],[214,5.463],[215,5.0401],[216,4.751],[217,4.5242],[218,4.6031],[219,4.6341],[220,4.6017],[221,4.8724],[222,5.0173],[224,5.1928],[225,5.3472],[226,5.6687],[227,6.0508],[228,6.6898],[229,7.5292],[230,8.5107],[231,10.0744],[232,11.5365],[233,14.1158],[234,14.9229],[235,13.345],[236,12.4223],[237,8.6549],[238,4.2093],[239,7.8847],[240,6.4124],[241,1.8685],[242,1.0583],[243,0.5753],[244,1.9655],[245,3.8412],[246,6.2035],[247,6.4043],[248,6.9084],[250,9.3581],[251,9.6366],[252,10.6654],[253,11.9484],[254,12.5561],[255,13.3011],[256,13.6298],[257,10.696],[258,-0.2031],[259,-1.6738],[260,-16.9204],[261,-26.3552],[262,-30.1715],[263,-31.0399],[264,-30.9721],[265,-30.5134],[266,-30.129],[267,-17.725],[268,-7.7119],[269,11.5241],[270,15.2125],[271,23.7391],[272,22.2411],[273,20.4994],[274,17.3498],[275,15.8258],[276,15.9178],[277,15.8914],[278,15.4801],[279,15.3882],[280,15.2569],[281,15.3509],[282,15.3765],[283,15.4849],[284,15.4883],[285,15.1894],[286,15.478],[289,15.5519],[290,15.5298],[291,15.0893],[292,13.9456],[293,13.2202],[294,13.6409],[295,14.2975],[296,14.3065],[297,14.1827],[298,11.3947],[299,9.4354],[300,6.3102],[301,6.1195],[302,4.5947],[303,4.3406],[304,3.575],[305,3.1304],[306,2.6069],[307,2.2634],[308,2.0184],[309,1.8711],[310,1.8235],[311,1.7333],[312,1.7011],[313,1.7127],[314,1.8172],[315,1.7798],[316,2.1226],[317,2.2098],[318,2.4781],[319,2.5431],[320,2.6684],[321,2.4143],[322,2.4901],[323,2.572],[324,2.3887],[325,2.4121],[326,2.5807],[327,2.7547],[328,2.6378],[329,2.8488],[330,3.0822],[331,3.1768],[332,3.1666],[333,3.5863],[334,3.7941],[335,3.9621],[336,4.0613],[337,4.037],[338,4.1964],[339,4.1362],[340,4.3268],[341,4.3077],[342,4.3742],[343,4.1828],[344,4.3126],[345,4.1732],[346,4.2847],[347,4.2658],[348,4.3563],[349,4.6284],[350,4.7399],[351,4.8127],[352,4.9028],[353,5.0444],[354,5.2075],[355,5.2932],[356,5.2847],[357,5.3654],[358,5.5589],[359,5.732],[360,5.7321]]
       }
       CTRL_C_eyesAim.Scale.X {type=MovieSceneScriptingFloatChannel, default=1, const [1,492]}
       CTRL_C_eyesAim.Scale.Y {type=MovieSceneScriptingFloatChannel, default=1, const [1,492]}
       CTRL_C_eyesAim.Scale.Z {type=MovieSceneScriptingFloatChannel, default=1, const [1,492]}
       CTRL_L_eyeAim.Location.X {type=MovieSceneScriptingFloatChannel, default=0, const [0,492]}
       CTRL_L_eyeAim.Location.Y {type=MovieSceneScriptingFloatChannel, default=0, const [0,492]}
       CTRL_L_eyeAim.Location.Z {type=MovieSceneScriptingFloatChannel, default=0, const [0,492]}
       CTRL_L_eyeAim.Rotation.X {type=MovieSceneScriptingFloatChannel, default=0, const [0,492]}
       CTRL_L_eyeAim.Rotation.Y {type=MovieSceneScriptingFloatChannel, default=0, const [0,492]}
       CTRL_L_eyeAim.Rotation.Z {type=MovieSceneScriptingFloatChannel, default=0, const [0,492]}
       CTRL_L_eyeAim.Scale.X {type=MovieSceneScriptingFloatChannel, default=1, const [1,492]}
       CTRL_L_eyeAim.Scale.Y {type=MovieSceneScriptingFloatChannel, default=1, const [1,492]}
       CTRL_L_eyeAim.Scale.Z {type=MovieSceneScriptingFloatChannel, default=1, const [1,492]}
       CTRL_R_eyeAim.Location.X {
        type=MovieSceneScriptingFloatChannel
        default=0
        thinned_from=492
        keys [[-131,0.0308],[-130,0.0136],[-129,0.0094],[-128,0.0084],[-127,-0.0011],[-126,0.0022],[-125,-0.0017],[-124,-0.0111],[-123,-0.0162],[-122,-0.0174],[-121,-0.0155],[-120,-0.0429],[-119,-0.0441],[-118,-0.0178],[-117,-0.0175],[-116,-0.0173],[-115,-0.0115],[-114,-0.0006],[-113,-0.0028],[-112,0.0005],[-111,0.0092],[-110,0.0147],[-109,0.0097],[-108,0.0192],[-107,0.0087],[-106,0.0064],[-105,0.0054],[-104,-0.0067],[-103,-0.0138],[-102,-0.0136],[-101,-0.016],[-100,-0.0203],[-99,-0.0146],[-98,-0.0225],[-97,-0.0191],[-96,-0.0316],[-95,-0.0365],[-94,-0.0275],[-93,-0.0269],[-92,-0.0215],[-91,-0.0384],[-90,-0.0333],[-89,-0.0287],[-88,-0.0396],[-87,-0.0481],[-86,-0.0527],[-85,-0.0518],[-84,-0.0512],[-83,-0.0545],[-82,-0.0527],[-81,-0.0588],[-80,-0.0722],[-79,-0.0722],[-78,-0.0711],[-77,-0.0362],[-76,-0.0154],[-75,-0.0025],[-74,0.01],[-73,0.0121],[-72,0.0116],[-71,0.0205],[-70,-0.0056],[-69,0.0211],[-68,0.009],[-67,0.014],[-66,0.0186],[-65,0.0148],[-64,0.0195],[-63,0.0169],[-62,0.0233],[-61,0.0275],[-60,0.0323],[-59,0.0261],[-58,0.0438],[-57,0.0441],[-56,0.0441],[-55,0.0382],[-54,0.043],[-53,0.0516],[-52,0.0525],[-51,0.0312],[-50,0.0266],[-49,0.0265],[-48,0.0144],[-47,0.0164],[-46,0.0133],[-45,0.0098],[-44,0.0068],[-43,-0.004],[-42,-0.0169],[-41,-0.0268],[-40,-0.0442],[-39,-0.0522],[-38,-0.0503],[-37,-0.0328],[-36,-0.0272],[-35,-0.001],[-34,0.0013],[-33,0.0138],[-32,0.0195],[-31,0.0256],[-30,0.0416],[-29,0.0425],[-28,0.0433],[-27,0.0431],[-26,0.0411],[-25,0.0464],[-24,0.0489],[-23,0.033],[-22,-0.0054],[-21,0.0004],[-20,0.0093],[-19,-0.0001],[-18,-0.0009],[-17,0.011],[-16,0.0142],[-15,0.012],[-14,-0.0038],[-13,-0.0061],[-12,-0.0063],[-11,0.0121],[-10,-0.0001],[-9,-0.0102],[-8,-0.0006],[-7,0.0136],[-6,0.0099],[-5,0.0007],[-4,-0.0098],[-3,0.0088],[-2,0.0038],[-1,0.0053],[0,0.0057],[1,-0.0027],[2,0.0028],[3,0.0078],[4,-0.0005],[5,0.0175],[6,-0.0013],[7,-0.0035],[8,-0.0073],[9,-0.0173],[10,-0.0169],[11,-0.0096],[12,-0.0072],[13,-0.0094],[14,-0.0182],[15,-0.0166],[16,-0.0197],[17,-0.0131],[18,-0.0181],[19,-0.0114],[20,-0.0092],[21,-0.0114],[22,-0.0104],[23,-0.013],[24,-0.0216],[25,-0.0262],[26,-0.0253],[27,-0.0183],[28,-0.0163],[29,-0.0151],[30,-0.0218],[31,-0.0193],[32,-0.0124],[33,-0.0236],[34,-0.0206],[35,-0.02],[36,-0.0167],[37,-0.0221],[38,-0.023],[39,-0.0268],[40,-0.0246],[41,-0.0321],[42,-0.0334],[43,-0.0304],[44,-0.0306],[45,-0.0301],[46,-0.0367],[47,-0.0237],[48,-0.0301],[49,-0.0202],[50,-0.007],[51,-0.0134],[52,-0.0264],[53,-0.0219],[54,-0.02],[55,-0.0141],[56,-0.016],[57,-0.0106],[58,-0.0106],[59,-0.0159],[60,-0.0169],[61,-0.008],[63,-0.009],[64,-0.0176],[65,-0.0176],[66,-0.0187],[67,-0.0076],[68,-0.0036],[69,-0.0029],[70,-0.0024],[71,-0.0015],[72,-0.0124],[73,0.0001],[74,-0.0049],[75,-0.0142],[76,0.0077],[77,0.0142],[78,0.013],[79,-0.0072],[80,-0.0097],[81,-0.0154],[82,-0.0016],[83,0.0013],[84,0.0007],[85,0.0007],[86,0.0109],[87,0.0144],[88,0.0148],[89,-0.0074],[90,0.0038],[91,0.0022],[92,-0.0022],[93,-0.0028],[94,-0.0072],[95,-0.0169],[96,-0.0187],[97,-0.0166],[98,-0.0107],[99,-0.017],[100,-0.0228],[101,-0.0178],[102,-0.0145],[103,-0.0217],[104,-0.0203],[105,-0.0201],[106,-0.0258],[107,-0.0374],[108,-0.048],[109,-0.0454],[110,-0.0582],[111,-0.0553],[112,-0.0679],[113,-0.0687],[114,-0.0676],[115,-0.0792],[116,-0.0879],[117,-0.0951],[118,-0.1173],[119,-0.1314],[120,-0.1435],[121,-0.1543],[122,-0.1594],[123,-0.1758],[124,-0.1799],[125,-0.1726],[126,-0.172],[127,-0.1934],[128,-0.1768],[129,-0.1683],[130,-0.1572],[131,-0.1727],[132,-0.1645],[133,-0.1495],[134,-0.1672],[135,-0.1938],[136,-0.2243],[137,-0.2788],[138,-0.3307],[139,-0.3722],[140,-0.4273],[141,-0.4344],[142,-0.4594],[143,-0.4452],[144,-0.4283],[145,-0.3904],[146,-0.4107],[147,-0.4074],[148,-0.3939],[149,-0.3954],[150,-0.4025],[151,-0.4181],[152,-0.4152],[153,-0.4161],[154,-0.4158],[155,-0.3829],[156,-0.3741],[157,-0.3568],[158,-0.3385],[159,-0.3262],[160,-0.3294],[161,-0.328],[162,-0.3169],[163,-0.3198],[164,-0.3354],[165,-0.3245],[166,-0.3118],[167,-0.3095],[168,-0.3004],[169,-0.2889],[170,-0.285],[171,-0.2699],[172,-0.2619],[173,-0.2417],[174,-0.2073],[175,-0.1934],[176,-0.1749],[177,-0.1738],[178,-0.1695],[179,-0.1618],[180,-0.1608],[181,-0.1694],[182,-0.1599],[184,-0.1185],[185,-0.1104],[186,-0.093],[187,-0.0556],[188,-0.0491],[189,-0.0105],[190,0.0334],[191,0.06],[192,-0.0394],[193,-0.1893],[194,-0.5139],[195,-0.746],[196,-1.1072],[197,-1.1403],[198,-0.8667],[199,-0.585],[200,-0.526],[201,-0.4151],[202,-0.274],[203,-0.2645],[204,-0.191],[205,-0.1504],[206,-0.1165],[207,-0.0793],[208,-0.0656],[209,-0.044],[210,-0.0334],[211,-0.0351],[212,-0.031],[213,-0.0259],[214,-0.0322],[215,-0.0163],[216,-0.0149],[217,-0.0147],[218,-0.0137],[219,-0.002],[220,0.0021],[221,-0.0067],[222,0.0026],[223,0.0038],[224,0.0155],[225,0.0184],[226,0.0062],[227,-0.0005],[228,-0.0188],[229,-0.0171],[230,-0.0233],[231,-0.0419],[232,-0.0747],[233,-0.1235],[234,-0.1118],[235,0.0213],[236,0.0612],[237,0.0066],[238,-0.0386],[239,0.057],[240,0.1397],[241,0.0514],[242,0.0162],[243,-0.0157],[244,0.0494],[245,0.1107],[246,0.1677],[247,0.1457],[248,0.1491],[249,0.1578],[250,0.1575],[251,0.1473],[252,0.1242],[253,0.1176],[254,0.085],[255,0.0775],[256,0.0744],[257,0.0245],[258,-0.0495],[259,-0.0592],[260,-0.1418],[261,-0.1899],[262,-0.1938],[263,-0.1027],[264,-0.1073],[265,-0.1096],[266,-0.1124],[267,-0.1454],[268,-0.1045],[269,-0.0243],[270,-0.001],[271,-0.0077],[272,0.0006],[274,-0.0028],[275,-0.0015],[276,0.0213],[277,0.0374],[278,0.0371],[279,0.0422],[280,0.0421],[281,0.031],[282,0.0373],[283,0.0379],[284,0.0375],[285,0.034],[286,0.0482],[287,0.0456],[289,0.0357],[290,0.0344],[291,0.0399],[292,0.0845],[293,0.0925],[294,0.0984],[295,0.1418],[296,0.1175],[297,0.11],[298,0.0603],[299,0.0366],[300,0.0061],[301,0.0067],[302,0.0154],[303,0.0187],[304,0.0275],[305,0.0271],[306,0.0249],[307,0.0234],[308,0.0297],[309,0.0274],[310,0.0246],[311,0.0076],[312,0.006],[313,0.0049],[314,0.0067],[315,0.0276],[316,0.0172],[317,0.0227],[318,0.0145],[319,0.0119],[320,0.0112],[321,0.0092],[322,0.0205],[323,0.0137],[324,0.0099],[325,0.0089],[327,0.0097],[328,0.009],[329,0.0092],[330,-0.0102],[331,0.003],[332,0.0087],[333,0.0021],[334,0.0113],[335,0.0143],[336,0.0158],[337,0.0133],[338,0.0254],[339,0.0173],[340,0.0147],[341,0.0144],[342,0.003],[343,0.0083],[344,0.0125],[346,0.014],[347,0.0032],[348,0.0114],[349,0.0104],[350,0.017],[351,0.0202],[352,0.0205],[353,0.0078],[354,0.0062],[355,0.0055],[356,0.0053],[357,0.0169],[358,0.0193],[359,0.0103],[360,0.0103]]
       }
       CTRL_R_eyeAim.Location.Y {
        type=MovieSceneScriptingFloatChannel
        default=0
        thinned_from=492
        keys [[-131,0.1735],[-130,0.1527],[-129,0.1444],[-128,0.1438],[-127,0.1001],[-126,0.067],[-125,0.0358],[-124,0.0169],[-123,0],[-122,-0.0044],[-121,0.0089],[-120,-0.0109],[-119,0.0077],[-118,0.0427],[-117,0.0457],[-116,0.0642],[-115,0.079],[-114,0.111],[-113,0.131],[-112,0.1426],[-111,0.1564],[-110,0.1739],[-109,0.1763],[-108,0.1679],[-107,0.1633],[-106,0.1612],[-105,0.1259],[-104,0.1171],[-103,0.0984],[-102,0.1066],[-101,0.0988],[-100,0.079],[-99,0.0701],[-98,0.057],[-97,0.0448],[-96,0.0374],[-95,0.0347],[-94,0.0187],[-93,0.018],[-92,0.009],[-91,0.0006],[-90,-0.0105],[-89,-0.0204],[-88,-0.0278],[-87,-0.0467],[-86,-0.0662],[-85,-0.0633],[-84,-0.0426],[-83,-0.0357],[-82,-0.0157],[-81,0.0075],[-80,0.022],[-79,0.024],[-78,0.0425],[-77,0.0789],[-76,0.0775],[-75,0.0873],[-74,0.0834],[-73,0.0735],[-72,0.0612],[-71,0.0562],[-70,0.0125],[-69,0.0041],[-68,0.0035],[-67,0.0007],[-66,0.0155],[-65,0.0455],[-64,0.0434],[-63,0.056],[-61,0.05],[-60,0.0501],[-59,0.0607],[-58,0.0688],[-57,0.0663],[-56,0.0583],[-55,0.0516],[-54,0.0458],[-53,0.0352],[-52,0.0301],[-51,0.0216],[-50,-0.0129],[-49,-0.017],[-48,-0.0172],[-47,-0.003],[-46,-0.0052],[-45,-0.0112],[-44,-0.0129],[-43,-0.014],[-42,0.001],[-41,0.0063],[-40,-0.0009],[-38,0.0065],[-37,0.0381],[-36,0.0414],[-35,0.0475],[-34,0.0725],[-33,0.0866],[-32,0.0983],[-31,0.1018],[-30,0.1069],[-29,0.1126],[-28,0.1203],[-27,0.1201],[-26,0.1117],[-25,0.1003],[-24,0.098],[-23,0.0886],[-22,0.2708],[-21,0.2725],[-20,0.1963],[-19,0.1085],[-18,0.0879],[-17,0.0886],[-16,0.0798],[-15,0.0632],[-14,0.0416],[-13,0.0452],[-12,0.043],[-11,0.0334],[-10,0.0347],[-9,0.0158],[-8,0.0173],[-6,0.0301],[-5,0.0003],[-4,0.0012],[-3,-0.0007],[-2,0.0032],[-1,0.0197],[0,0.0247],[1,0.0285],[2,0.0331],[3,0.0463],[5,0.0566],[6,0.059],[7,0.0553],[8,0.054],[9,0.0504],[10,0.0472],[11,0.0455],[12,0.0466],[13,0.0529],[14,0.0497],[15,0.0357],[16,0.0281],[17,0.0376],[18,0.0396],[19,0.0393],[20,0.0371],[21,0.0241],[22,0.0265],[23,0.0281],[24,0.0302],[25,0.0243],[26,0.031],[27,0.0396],[28,0.0384],[29,0.0342],[30,0.0185],[31,0.0114],[32,0.0222],[33,0.0138],[34,-0.0014],[35,-0.0002],[36,0.0096],[37,0.0097],[38,0.003],[39,0.0029],[40,-0.0044],[41,-0.0079],[42,-0.0215],[43,-0.0376],[44,-0.048],[45,-0.0475],[46,-0.0762],[47,-0.0668],[48,-0.0754],[49,-0.0665],[50,-0.0694],[51,-0.0693],[52,-0.0926],[53,-0.1153],[54,-0.1198],[55,-0.1097],[56,-0.0851],[57,-0.0438],[58,-0.0046],[59,-0.0113],[60,-0.0158],[61,-0.0385],[62,-0.0417],[63,-0.0408],[64,-0.027],[65,-0.027],[66,-0.0363],[67,-0.027],[68,-0.0273],[69,-0.033],[70,-0.028],[71,-0.0163],[72,-0.0134],[73,0.0008],[74,0.0148],[75,-0.0037],[76,0.0063],[77,0.0011],[78,0.0013],[79,0.0045],[80,0.0223],[81,0.0015],[82,0.0211],[83,0.0169],[84,0.0072],[86,-0.0016],[87,-0.0036],[88,-0.007],[89,-0.0264],[90,-0.0263],[91,-0.0572],[92,-0.077],[93,-0.0832],[94,-0.1002],[95,-0.1072],[96,-0.1032],[97,-0.1012],[98,-0.0934],[99,-0.1031],[100,-0.1264],[101,-0.1343],[102,-0.1339],[103,-0.1488],[104,-0.1653],[105,-0.1662],[106,-0.1948],[107,-0.2135],[108,-0.2443],[109,-0.2728],[110,-0.2868],[111,-0.3118],[112,-0.3295],[113,-0.3692],[114,-0.3722],[115,-0.4087],[116,-0.443],[117,-0.4694],[118,-0.511],[119,-0.5423],[120,-0.5767],[121,-0.5929],[122,-0.5944],[123,-0.6202],[124,-0.6358],[125,-0.6328],[126,-0.6428],[127,-0.6467],[128,-0.643],[129,-0.636],[130,-0.6437],[131,-0.6403],[132,-0.6348],[133,-0.6549],[134,-0.6985],[135,-0.7273],[136,-0.7403],[137,-0.8023],[138,-0.8277],[139,-0.7692],[140,-0.7667],[141,-0.7578],[142,-0.7589],[144,-0.7146],[145,-0.6939],[146,-0.6806],[147,-0.691],[148,-0.6843],[149,-0.699],[150,-0.7097],[151,-0.7221],[152,-0.7147],[154,-0.7166],[155,-0.687],[156,-0.6757],[157,-0.6667],[158,-0.6477],[159,-0.6396],[160,-0.6124],[161,-0.5996],[162,-0.5787],[163,-0.581],[164,-0.5775],[165,-0.5701],[166,-0.5488],[167,-0.533],[168,-0.5039],[169,-0.4993],[170,-0.4878],[171,-0.4776],[172,-0.4662],[173,-0.4434],[174,-0.4143],[175,-0.3921],[176,-0.3841],[177,-0.3813],[178,-0.3668],[179,-0.347],[180,-0.3481],[181,-0.3397],[182,-0.3263],[183,-0.2993],[184,-0.2778],[185,-0.2593],[186,-0.2172],[187,-0.1382],[188,-0.1235],[189,-0.0279],[190,0.0999],[191,0.2183],[192,-0.0347],[193,-0.423],[194,-0.9419],[195,-1.3603],[196,-2.0677],[197,-2.3127],[198,-2.1616],[199,-1.8214],[200,-1.7222],[201,-1.5583],[202,-1.4682],[203,-1.1794],[204,-0.958],[205,-0.7427],[206,-0.5992],[207,-0.4769],[208,-0.3437],[209,-0.2587],[210,-0.2275],[211,-0.1677],[212,-0.1331],[213,-0.0841],[214,-0.0477],[215,-0.0426],[216,-0.0169],[217,0.0127],[218,0.0303],[219,0.0379],[220,0.0504],[221,0.0619],[222,0.0728],[223,0.057],[224,0.0624],[225,0.0438],[226,0.0193],[227,-0.0189],[228,-0.0714],[229,-0.1037],[230,-0.1476],[231,-0.2332],[232,-0.312],[233,-0.4285],[234,-0.3578],[235,0.2133],[236,0.5487],[237,0.5128],[238,0.5362],[239,1.0621],[240,1.8151],[241,2.7253],[242,2.8943],[243,2.8814],[244,2.5618],[246,1.8184],[247,1.7122],[248,1.593],[250,1.2259],[251,1.0777],[252,0.8977],[253,0.7791],[254,0.6497],[255,0.5511],[256,0.4737],[257,0.3878],[258,0.3196],[259,0.3111],[260,0.2207],[261,0.1761],[262,0.1378],[263,-0.0224],[264,-0.0192],[265,-0.0171],[266,-0.0023],[267,0.1323],[268,0.0984],[269,0.1508],[270,0.1734],[271,0.0946],[272,0.0683],[273,0.0808],[274,0.1055],[275,0.1418],[276,0.2019],[277,0.2512],[278,0.2542],[279,0.2532],[280,0.253],[281,0.2556],[282,0.2727],[283,0.2612],[284,0.2586],[285,0.2487],[286,0.2607],[287,0.2605],[288,0.257],[289,0.2457],[290,0.2423],[291,0.2718],[292,0.5425],[293,0.7078],[294,0.7018],[295,0.7433],[296,0.6427],[297,0.6187],[298,0.4571],[299,0.3631],[300,0.2479],[301,0.2505],[302,0.2633],[303,0.2602],[304,0.2437],[306,0.1878],[307,0.1708],[308,0.1606],[309,0.1551],[310,0.1538],[311,0.1499],[312,0.145],[313,0.1462],[314,0.1582],[315,0.1779],[316,0.2029],[317,0.2259],[318,0.2259],[319,0.2306],[320,0.2272],[321,0.2101],[322,0.2309],[323,0.2286],[324,0.2109],[325,0.2031],[326,0.2046],[328,0.1956],[329,0.204],[330,0.2096],[331,0.2045],[332,0.2043],[333,0.2142],[334,0.2276],[335,0.2308],[336,0.2291],[337,0.1902],[338,0.1823],[339,0.1769],[340,0.177],[341,0.1833],[342,0.1892],[343,0.1746],[345,0.1883],[346,0.1961],[347,0.1912],[348,0.2004],[349,0.2587],[350,0.265],[351,0.2563],[352,0.2441],[353,0.2412],[354,0.2391],[355,0.2355],[356,0.2347],[357,0.2276],[358,0.2137],[359,0.2212],[360,0.2212]]
       }
       CTRL_R_eyeAim.Location.Z {
        type=MovieSceneScriptingFloatChannel
        default=0
        thinned_from=492
        keys [[-131,-0.5387],[-130,-0.5397],[-129,-0.5273],[-128,-0.5385],[-127,-0.5217],[-126,-0.4967],[-125,-0.4858],[-124,-0.4935],[-122,-0.4937],[-121,-0.4768],[-120,-0.4622],[-119,-0.4654],[-118,-0.433],[-117,-0.4336],[-116,-0.4395],[-115,-0.4349],[-114,-0.4521],[-113,-0.4792],[-112,-0.5018],[-111,-0.5202],[-110,-0.5328],[-109,-0.5424],[-108,-0.5328],[-106,-0.5336],[-105,-0.5266],[-104,-0.5584],[-103,-0.544],[-102,-0.5747],[-101,-0.535],[-100,-0.5495],[-99,-0.5466],[-98,-0.55],[-97,-0.5633],[-96,-0.5854],[-95,-0.5913],[-94,-0.5802],[-93,-0.5793],[-92,-0.5664],[-91,-0.5623],[-90,-0.5553],[-89,-0.5413],[-88,-0.5394],[-87,-0.5019],[-86,-0.4714],[-85,-0.4579],[-84,-0.4534],[-83,-0.4433],[-82,-0.4307],[-81,-0.4465],[-80,-0.4481],[-79,-0.4425],[-78,-0.3958],[-77,-0.42],[-76,-0.4385],[-75,-0.3932],[-74,-0.4263],[-73,-0.4632],[-72,-0.443],[-71,-0.4573],[-70,-0.4301],[-69,-0.432],[-68,-0.4328],[-67,-0.4165],[-66,-0.4273],[-65,-0.4302],[-64,-0.4227],[-63,-0.4022],[-62,-0.4122],[-61,-0.4326],[-60,-0.4376],[-59,-0.4297],[-58,-0.4448],[-56,-0.435],[-55,-0.434],[-54,-0.4182],[-53,-0.4231],[-52,-0.4306],[-51,-0.454],[-50,-0.4473],[-49,-0.446],[-48,-0.4088],[-47,-0.4115],[-46,-0.4094],[-45,-0.4087],[-44,-0.4062],[-43,-0.3959],[-42,-0.3993],[-41,-0.4225],[-40,-0.4444],[-39,-0.4466],[-38,-0.4435],[-37,-0.4305],[-36,-0.4542],[-35,-0.4451],[-34,-0.4564],[-33,-0.4445],[-32,-0.4404],[-31,-0.4384],[-30,-0.406],[-29,-0.3868],[-28,-0.3632],[-27,-0.3987],[-26,-0.3936],[-25,-0.3716],[-24,-0.3654],[-23,-0.3464],[-22,0.6844],[-21,0.5977],[-20,-0.461],[-19,-0.5615],[-18,-0.5267],[-17,-0.4502],[-16,-0.3945],[-15,-0.3445],[-14,-0.3202],[-13,-0.3242],[-12,-0.2982],[-11,-0.3007],[-10,-0.3013],[-9,-0.2819],[-8,-0.2963],[-7,-0.2782],[-6,-0.2589],[-5,-0.265],[-4,-0.2469],[-3,-0.2497],[-2,-0.2351],[-1,-0.2456],[0,-0.2339],[1,-0.2362],[2,-0.2616],[3,-0.2817],[4,-0.294],[5,-0.2622],[6,-0.2598],[8,-0.2596],[9,-0.2582],[10,-0.2577],[11,-0.2664],[12,-0.2699],[13,-0.2616],[14,-0.2545],[15,-0.2685],[16,-0.2831],[17,-0.2844],[18,-0.294],[19,-0.3084],[20,-0.3179],[21,-0.3382],[22,-0.3024],[23,-0.2977],[24,-0.2878],[25,-0.2876],[26,-0.2952],[27,-0.3107],[28,-0.3271],[29,-0.3259],[30,-0.3198],[31,-0.3059],[32,-0.282],[33,-0.2659],[34,-0.2815],[35,-0.2787],[36,-0.2628],[37,-0.2834],[38,-0.2865],[39,-0.2947],[40,-0.2955],[41,-0.2746],[42,-0.2604],[43,-0.2545],[44,-0.2539],[45,-0.258],[46,-0.2525],[47,-0.2754],[48,-0.281],[49,-0.317],[50,-0.3464],[51,-0.3714],[52,-0.4204],[53,-0.4717],[54,-0.4872],[55,-0.5249],[56,-0.5257],[57,-0.5536],[58,-0.563],[59,-0.5705],[60,-0.5732],[61,-0.5562],[62,-0.5817],[63,-0.583],[64,-0.6036],[65,-0.6036],[66,-0.6347],[68,-0.6388],[69,-0.6612],[70,-0.6355],[71,-0.6419],[72,-0.6216],[73,-0.6171],[74,-0.5948],[75,-0.5785],[76,-0.5429],[77,-0.5359],[78,-0.5364],[79,-0.5454],[80,-0.5265],[81,-0.558],[82,-0.5201],[84,-0.5223],[85,-0.517],[86,-0.526],[87,-0.5187],[88,-0.5416],[89,-0.5167],[90,-0.5237],[91,-0.513],[92,-0.5328],[93,-0.5235],[94,-0.5015],[95,-0.4964],[96,-0.4978],[97,-0.5053],[98,-0.4946],[99,-0.5033],[100,-0.5236],[101,-0.5128],[102,-0.5066],[103,-0.5121],[104,-0.5256],[105,-0.5265],[106,-0.5028],[107,-0.52],[108,-0.5021],[109,-0.4613],[110,-0.4731],[111,-0.4597],[112,-0.4384],[113,-0.4285],[114,-0.4521],[115,-0.4627],[116,-0.4641],[117,-0.4738],[118,-0.4628],[119,-0.4422],[120,-0.4454],[121,-0.4618],[122,-0.4308],[123,-0.3717],[124,-0.3676],[125,-0.3579],[126,-0.3728],[127,-0.3843],[128,-0.378],[129,-0.3977],[130,-0.3814],[131,-0.3821],[132,-0.3792],[133,-0.3866],[134,-0.4056],[135,-0.4032],[136,-0.4226],[137,-0.3751],[138,-0.3588],[139,-0.2431],[140,-0.1784],[141,-0.1429],[142,-0.1287],[143,-0.1292],[144,-0.1727],[145,-0.1998],[146,-0.2098],[147,-0.1978],[148,-0.1732],[149,-0.1973],[150,-0.2003],[151,-0.1906],[152,-0.1899],[153,-0.2242],[154,-0.2291],[155,-0.2419],[156,-0.2411],[157,-0.242],[158,-0.23],[159,-0.208],[160,-0.2021],[161,-0.1949],[162,-0.1827],[163,-0.2119],[164,-0.1741],[165,-0.1788],[166,-0.1629],[167,-0.163],[168,-0.1614],[169,-0.1927],[170,-0.1941],[171,-0.1752],[172,-0.1771],[173,-0.195],[174,-0.2101],[175,-0.2144],[176,-0.2409],[177,-0.2379],[178,-0.2278],[179,-0.2248],[180,-0.2097],[181,-0.2147],[182,-0.2075],[183,-0.2284],[184,-0.2292],[185,-0.2242],[186,-0.2332],[187,-0.2567],[188,-0.26],[189,-0.2416],[190,-0.2571],[191,-0.2527],[192,-0.5156],[193,-0.4083],[194,-0.851],[195,-0.55],[196,0.2074],[197,0.4932],[198,0.3551],[199,0.1092],[200,0.0461],[201,-0.064],[202,-0.0801],[203,-0.1112],[204,-0.1699],[205,-0.2629],[206,-0.3233],[207,-0.3946],[208,-0.4215],[209,-0.4294],[210,-0.3864],[211,-0.3738],[212,-0.3731],[213,-0.3918],[214,-0.4042],[215,-0.4053],[216,-0.4215],[217,-0.4218],[218,-0.4156],[219,-0.3973],[220,-0.4017],[221,-0.387],[222,-0.3877],[223,-0.3837],[224,-0.3866],[225,-0.3832],[226,-0.3812],[227,-0.3751],[228,-0.3494],[229,-0.3354],[230,-0.3243],[231,-0.3155],[232,-0.2885],[233,-0.3148],[234,-0.3712],[235,-0.5021],[236,0.0169],[237,1.1873],[238,1.5733],[239,2.0164],[240,1.8286],[241,0.9429],[242,0.7802],[243,0.4325],[244,0.227],[245,0.0743],[246,-0.1132],[247,-0.1242],[248,-0.2162],[249,-0.3221],[250,-0.3713],[251,-0.4252],[252,-0.4264],[253,-0.4251],[254,-0.4566],[255,-0.4381],[256,-0.4299],[257,-0.5176],[258,-0.6344],[259,-0.631],[260,-0.3847],[261,0.2541],[262,0.7936],[263,0.7248],[264,0.7691],[265,0.7759],[266,0.5977],[267,0.6912],[268,0.8849],[269,-0.1502],[270,-0.3664],[271,-0.4646],[272,-0.4804],[273,-0.4748],[274,-0.4746],[275,-0.4603],[276,-0.4483],[277,-0.4592],[278,-0.4571],[279,-0.4521],[280,-0.4346],[281,-0.437],[282,-0.4404],[283,-0.4326],[284,-0.4319],[285,-0.4414],[286,-0.4557],[287,-0.4513],[288,-0.454],[289,-0.4787],[290,-0.4907],[291,-0.4923],[292,-0.5049],[293,0.0587],[294,-0.2057],[295,-0.6828],[296,-0.6648],[297,-0.6537],[298,-0.6237],[299,-0.6121],[300,-0.6231],[301,-0.623],[302,-0.6417],[303,-0.6484],[304,-0.6649],[305,-0.6665],[306,-0.6802],[307,-0.6999],[308,-0.7053],[309,-0.7229],[310,-0.7311],[311,-0.7171],[312,-0.7074],[313,-0.7058],[314,-0.7159],[315,-0.7132],[316,-0.7086],[317,-0.6993],[318,-0.7067],[320,-0.7099],[321,-0.7213],[322,-0.7039],[323,-0.6744],[324,-0.6883],[325,-0.692],[326,-0.6706],[327,-0.6741],[328,-0.6433],[329,-0.6545],[330,-0.6682],[331,-0.6531],[332,-0.6355],[333,-0.6081],[335,-0.6089],[336,-0.5975],[337,-0.596],[338,-0.5831],[339,-0.6007],[340,-0.5963],[341,-0.5853],[342,-0.6027],[343,-0.588],[344,-0.581],[345,-0.5897],[346,-0.6045],[347,-0.6087],[348,-0.6],[349,-0.6237],[350,-0.6149],[351,-0.6329],[352,-0.6177],[353,-0.613],[354,-0.6051],[355,-0.6074],[357,-0.6079],[358,-0.5978],[359,-0.595],[360,-0.595]]
       }
       CTRL_R_eyeAim.Rotation.X {
        type=MovieSceneScriptingFloatChannel
        default=0
        thinned_from=492
        keys [[-131,-1.2714],[-130,-1.272],[-129,-1.2425],[-128,-1.2687],[-127,-1.2269],[-126,-1.167],[-125,-1.1408],[-124,-1.1588],[-122,-1.1593],[-121,-1.1195],[-120,-1.0853],[-119,-1.0928],[-118,-1.0169],[-117,-1.0184],[-116,-1.0328],[-115,-1.0223],[-114,-1.064],[-113,-1.1286],[-112,-1.1824],[-111,-1.2263],[-110,-1.2572],[-109,-1.2799],[-108,-1.2568],[-106,-1.258],[-105,-1.2395],[-104,-1.3137],[-103,-1.279],[-102,-1.3516],[-101,-1.258],[-100,-1.2914],[-99,-1.2844],[-98,-1.2921],[-97,-1.3228],[-96,-1.3747],[-95,-1.3885],[-94,-1.3623],[-93,-1.3603],[-92,-1.3298],[-91,-1.3203],[-90,-1.3038],[-89,-1.2711],[-88,-1.2668],[-87,-1.1791],[-86,-1.1081],[-85,-1.0761],[-84,-1.0651],[-83,-1.0411],[-82,-1.0112],[-81,-1.0484],[-80,-1.0522],[-79,-1.0391],[-78,-0.9295],[-77,-0.9873],[-76,-1.0308],[-75,-0.9248],[-74,-1.0023],[-73,-1.0886],[-72,-1.0408],[-71,-1.0742],[-70,-1.0098],[-69,-1.0144],[-68,-1.0163],[-67,-0.9779],[-66,-1.0033],[-65,-1.0106],[-64,-0.9927],[-63,-0.945],[-62,-0.9684],[-61,-1.0162],[-60,-1.0278],[-59,-1.0097],[-58,-1.0452],[-56,-1.022],[-55,-1.0194],[-54,-0.9821],[-53,-0.9937],[-52,-1.0112],[-51,-1.0661],[-50,-1.0503],[-49,-1.0472],[-48,-0.9599],[-47,-0.9663],[-46,-0.9613],[-45,-0.9596],[-44,-0.9538],[-43,-0.9296],[-42,-0.9375],[-41,-0.992],[-40,-1.0433],[-39,-1.0485],[-38,-1.0412],[-37,-1.0109],[-36,-1.0666],[-35,-1.0456],[-34,-1.0726],[-33,-1.0452],[-32,-1.0359],[-31,-1.0315],[-30,-0.9557],[-29,-0.9107],[-28,-0.8559],[-27,-0.9391],[-26,-0.9266],[-25,-0.8746],[-24,-0.8598],[-23,-0.8149],[-22,1.5921],[-21,1.3877],[-20,-1.0904],[-19,-1.3207],[-18,-1.2382],[-17,-1.0585],[-16,-0.9275],[-15,-0.8097],[-14,-0.7522],[-13,-0.7617],[-12,-0.7006],[-11,-0.7063],[-10,-0.7077],[-9,-0.6619],[-8,-0.6957],[-7,-0.6533],[-6,-0.6081],[-5,-0.6223],[-4,-0.5797],[-3,-0.5862],[-2,-0.552],[-1,-0.5767],[0,-0.5493],[1,-0.5548],[2,-0.6144],[3,-0.6619],[4,-0.6907],[5,-0.6163],[6,-0.6107],[8,-0.6102],[9,-0.6067],[10,-0.6055],[11,-0.6259],[12,-0.6341],[13,-0.6148],[14,-0.598],[15,-0.6305],[16,-0.6647],[17,-0.668],[18,-0.6906],[19,-0.7243],[20,-0.7466],[21,-0.7942],[22,-0.7102],[23,-0.699],[24,-0.6758],[25,-0.6752],[26,-0.6933],[27,-0.7298],[28,-0.7683],[29,-0.7654],[30,-0.751],[31,-0.7182],[32,-0.6622],[33,-0.6242],[34,-0.661],[35,-0.6543],[36,-0.617],[37,-0.6655],[38,-0.6727],[39,-0.6918],[40,-0.6937],[41,-0.6447],[42,-0.6114],[43,-0.5978],[44,-0.5968],[45,-0.6062],[46,-0.5941],[47,-0.6477],[48,-0.6612],[49,-0.7453],[50,-0.8145],[51,-0.8733],[52,-0.9891],[53,-1.1106],[54,-1.1473],[55,-1.2354],[56,-1.2362],[57,-1.3005],[58,-1.322],[59,-1.3397],[60,-1.346],[61,-1.3065],[62,-1.3664],[63,-1.3694],[64,-1.4175],[65,-1.4175],[66,-1.4909],[68,-1.5003],[69,-1.553],[70,-1.4926],[71,-1.5074],[72,-1.4597],[73,-1.449],[74,-1.3967],[75,-1.3585],[76,-1.2748],[77,-1.2582],[78,-1.2596],[79,-1.2806],[80,-1.2362],[81,-1.3101],[82,-1.2212],[84,-1.2264],[85,-1.2139],[86,-1.2351],[87,-1.2178],[88,-1.2716],[89,-1.2134],[90,-1.2298],[91,-1.2054],[92,-1.2526],[93,-1.231],[94,-1.1801],[95,-1.1684],[96,-1.1715],[97,-1.189],[98,-1.1635],[99,-1.1844],[100,-1.2334],[101,-1.2084],[102,-1.1938],[103,-1.2077],[104,-1.2406],[105,-1.2428],[106,-1.1896],[107,-1.2317],[108,-1.1928],[109,-1.1005],[110,-1.1301],[111,-1.1018],[112,-1.0544],[113,-1.0375],[114,-1.0936],[115,-1.1253],[116,-1.1353],[117,-1.1636],[118,-1.1473],[119,-1.1065],[120,-1.1231],[121,-1.1666],[122,-1.0942],[123,-0.963],[124,-0.9581],[125,-0.9347],[126,-0.9728],[127,-1.0008],[128,-0.985],[129,-1.0291],[130,-0.9931],[131,-0.9938],[132,-0.985],[133,-1.0078],[134,-1.0664],[135,-1.0706],[136,-1.123],[137,-1.0393],[138,-1.0213],[139,-0.7327],[140,-0.5867],[141,-0.5021],[142,-0.4729],[143,-0.4635],[144,-0.5554],[145,-0.6098],[146,-0.6282],[147,-0.6031],[148,-0.5433],[149,-0.6052],[150,-0.6169],[151,-0.599],[152,-0.5937],[153,-0.6749],[154,-0.6864],[155,-0.7038],[156,-0.6976],[157,-0.6959],[158,-0.6602],[159,-0.6046],[160,-0.5821],[161,-0.561],[162,-0.5252],[163,-0.5949],[164,-0.5059],[165,-0.5142],[166,-0.4705],[167,-0.4666],[168,-0.4553],[169,-0.5274],[170,-0.527],[171,-0.4783],[172,-0.4784],[173,-0.5132],[174,-0.5406],[175,-0.545],[176,-0.6052],[177,-0.5973],[178,-0.5702],[179,-0.5596],[180,-0.5242],[181,-0.5342],[182,-0.5151],[183,-0.5597],[184,-0.5583],[185,-0.5439],[186,-0.5597],[187,-0.6077],[188,-0.6142],[189,-0.5673],[190,-0.6058],[191,-0.6043],[192,-1.211],[193,-1.0026],[194,-2.2439],[195,-1.806],[196,-0.7172],[197,-0.3165],[198,-0.3637],[199,-0.5408],[200,-0.5954],[201,-0.7137],[202,-0.6706],[203,-0.5741],[204,-0.6035],[205,-0.7391],[206,-0.8381],[207,-0.9768],[208,-1.0159],[209,-1.0232],[210,-0.9188],[211,-0.884],[212,-0.8802],[213,-0.9216],[214,-0.9497],[215,-0.952],[216,-0.9897],[217,-0.9905],[218,-0.976],[219,-0.9332],[220,-0.9437],[221,-0.9093],[222,-0.9113],[223,-0.9015],[224,-0.9083],[225,-0.9002],[226,-0.8951],[227,-0.8808],[228,-0.8215],[229,-0.7899],[230,-0.7664],[231,-0.753],[232,-0.6993],[233,-0.7815],[234,-0.9014],[235,-1.1887],[236,-0.028],[237,2.7363],[238,3.6441],[239,4.5228],[240,3.5968],[241,0.5128],[242,-0.1039],[243,-0.8955],[244,-0.9521],[245,-0.8962],[246,-0.9988],[247,-0.9398],[249,-1.1939],[250,-1.2027],[251,-1.2531],[252,-1.1783],[253,-1.1322],[254,-1.1654],[255,-1.0958],[256,-1.059],[257,-1.2474],[258,-1.5103],[259,-1.5011],[260,-0.9142],[261,0.5884],[262,1.8596],[263,1.7022],[264,1.8064],[265,1.8222],[266,1.4038],[267,1.6196],[268,2.0765],[269,-0.3577],[270,-0.8668],[271,-1.0929],[272,-1.1288],[273,-1.1162],[274,-1.1166],[275,-1.0851],[276,-1.0615],[277,-1.0922],[278,-1.0875],[279,-1.0756],[280,-1.0346],[281,-1.0404],[282,-1.0505],[283,-1.0308],[284,-1.0288],[285,-1.05],[286,-1.0849],[287,-1.0746],[288,-1.0804],[289,-1.1371],[290,-1.1649],[291,-1.1721],[292,-1.2506],[293,0.0243],[294,-0.5943],[295,-1.7254],[296,-1.6519],[297,-1.6191],[298,-1.509],[299,-1.4646],[300,-1.4755],[301,-1.4755],[302,-1.5205],[303,-1.536],[304,-1.5729],[305,-1.574],[306,-1.604],[307,-1.6488],[308,-1.6609],[309,-1.7019],[310,-1.7211],[311,-1.6881],[312,-1.665],[313,-1.6613],[314,-1.6857],[315,-1.6807],[316,-1.6716],[317,-1.6519],[318,-1.6693],[319,-1.6751],[320,-1.6769],[321,-1.702],[322,-1.6631],[323,-1.5936],[324,-1.6248],[325,-1.6327],[326,-1.5828],[327,-1.5906],[328,-1.5179],[329,-1.5449],[330,-1.5774],[331,-1.5417],[332,-1.5004],[333,-1.4369],[335,-1.4402],[336,-1.4134],[337,-1.4066],[338,-1.3757],[339,-1.4166],[340,-1.4062],[341,-1.3809],[342,-1.4221],[343,-1.3865],[344,-1.3706],[345,-1.3915],[346,-1.427],[347,-1.4364],[348,-1.4167],[349,-1.4778],[350,-1.4578],[351,-1.4992],[352,-1.4622],[353,-1.4511],[354,-1.4323],[355,-1.4372],[357,-1.4377],[358,-1.4126],[359,-1.4068],[360,-1.4068]]
       }
       CTRL_R_eyeAim.Rotation.Y {
        type=MovieSceneScriptingFloatChannel
        default=0
        thinned_from=492
        keys [[-131,0.3944],[-130,0.3616],[-129,0.3589],[-128,0.3632],[-127,0.3325],[-126,0.2982],[-125,0.2838],[-124,0.2757],[-122,0.2773],[-121,0.2634],[-120,0.2489],[-119,0.2502],[-118,0.2336],[-117,0.2337],[-116,0.2349],[-115,0.2296],[-114,0.2306],[-113,0.2347],[-112,0.2315],[-111,0.2299],[-110,0.2342],[-109,0.2478],[-107,0.2528],[-106,0.2532],[-105,0.2523],[-104,0.256],[-103,0.2371],[-102,0.2427],[-101,0.2163],[-100,0.2108],[-99,0.1999],[-98,0.1887],[-97,0.1831],[-95,0.181],[-94,0.1681],[-93,0.1673],[-92,0.1645],[-91,0.169],[-90,0.172],[-89,0.1797],[-88,0.1901],[-87,0.1905],[-86,0.1894],[-85,0.1931],[-84,0.1999],[-82,0.2006],[-81,0.2085],[-80,0.2098],[-79,0.2078],[-78,0.1908],[-77,0.2071],[-76,0.2135],[-75,0.1967],[-74,0.2031],[-73,0.2117],[-72,0.1954],[-71,0.1964],[-70,0.1884],[-69,0.1885],[-68,0.189],[-67,0.1844],[-66,0.1904],[-65,0.1945],[-64,0.1913],[-63,0.1874],[-62,0.1916],[-61,0.2016],[-60,0.2046],[-59,0.1992],[-58,0.2111],[-57,0.2087],[-56,0.2057],[-55,0.2058],[-54,0.1969],[-53,0.1988],[-52,0.2036],[-51,0.2158],[-50,0.2177],[-49,0.2176],[-48,0.209],[-47,0.2206],[-46,0.2208],[-45,0.2244],[-44,0.2238],[-43,0.2211],[-42,0.2275],[-41,0.242],[-40,0.2486],[-39,0.2413],[-38,0.2371],[-37,0.2244],[-36,0.2322],[-35,0.2257],[-34,0.2331],[-33,0.2252],[-32,0.222],[-31,0.2221],[-30,0.2078],[-29,0.1984],[-28,0.1868],[-27,0.206],[-26,0.2029],[-25,0.187],[-24,0.1812],[-23,0.1035],[-22,-0.1353],[-21,-0.2168],[-20,0.1931],[-19,0.2344],[-18,0.2247],[-17,0.2011],[-16,0.1835],[-15,0.1611],[-14,0.1474],[-13,0.1493],[-12,0.1385],[-11,0.1366],[-10,0.1393],[-9,0.1324],[-8,0.1382],[-7,0.1326],[-6,0.124],[-5,0.1341],[-4,0.1256],[-3,0.1295],[-2,0.1227],[-1,0.1281],[0,0.119],[1,0.1203],[2,0.131],[3,0.1425],[4,0.1484],[5,0.1343],[6,0.1343],[7,0.1354],[8,0.1352],[9,0.1342],[10,0.1359],[11,0.1422],[12,0.1447],[13,0.1407],[14,0.1387],[15,0.147],[16,0.1545],[17,0.1555],[18,0.1627],[19,0.1716],[20,0.1788],[21,0.1872],[22,0.1702],[23,0.168],[24,0.1628],[25,0.162],[26,0.1688],[27,0.1787],[28,0.1895],[29,0.1904],[30,0.1901],[31,0.1869],[32,0.1743],[33,0.1646],[34,0.1755],[35,0.1743],[36,0.168],[37,0.178],[38,0.1775],[39,0.182],[40,0.1853],[41,0.1728],[42,0.1622],[43,0.1604],[44,0.1615],[45,0.1638],[46,0.1585],[47,0.1697],[48,0.1651],[49,0.1797],[50,0.1862],[51,0.1763],[52,0.1822],[53,0.1567],[54,0.1561],[55,0.1457],[56,0.1226],[57,0.1161],[58,0.1164],[59,0.1196],[60,0.1205],[61,0.115],[62,0.1184],[63,0.1178],[64,0.1086],[65,0.1086],[66,0.1038],[67,0.1036],[69,0.1121],[70,0.121],[71,0.131],[72,0.1363],[73,0.1647],[74,0.1741],[75,0.1791],[76,0.1853],[77,0.2],[78,0.2007],[79,0.2118],[80,0.2153],[81,0.2356],[82,0.2218],[83,0.2273],[84,0.2336],[85,0.236],[86,0.2409],[87,0.2399],[88,0.258],[89,0.2492],[90,0.2492],[91,0.2454],[92,0.2488],[93,0.2443],[94,0.2343],[95,0.2316],[96,0.2263],[97,0.2289],[98,0.2208],[99,0.2204],[100,0.231],[101,0.2303],[102,0.2288],[103,0.2332],[104,0.2456],[105,0.2464],[106,0.2373],[107,0.2482],[108,0.2451],[109,0.2301],[110,0.2442],[111,0.2387],[112,0.2304],[113,0.2296],[114,0.2443],[115,0.2636],[116,0.2684],[117,0.2766],[118,0.2758],[119,0.2721],[120,0.2806],[121,0.3039],[122,0.2934],[123,0.2729],[125,0.27],[126,0.2822],[127,0.2882],[128,0.2849],[129,0.2956],[130,0.2851],[131,0.2865],[132,0.2778],[133,0.2731],[134,0.286],[135,0.2949],[136,0.344],[137,0.3776],[138,0.4569],[139,0.377],[140,0.333],[141,0.295],[142,0.2894],[143,0.2816],[144,0.3333],[145,0.3622],[146,0.3736],[147,0.354],[148,0.3223],[149,0.3558],[150,0.3644],[151,0.3513],[152,0.3464],[153,0.3925],[155,0.4041],[157,0.4],[158,0.3776],[159,0.3401],[160,0.3341],[161,0.3245],[162,0.3036],[163,0.3452],[164,0.2993],[165,0.3035],[166,0.2808],[167,0.2855],[168,0.2872],[169,0.3317],[170,0.3294],[171,0.288],[172,0.2776],[173,0.2827],[174,0.2848],[175,0.2748],[176,0.2961],[177,0.2893],[178,0.2682],[179,0.2621],[180,0.2426],[181,0.2456],[182,0.2366],[183,0.2566],[184,0.2542],[185,0.2437],[186,0.2443],[187,0.2468],[188,0.2461],[189,0.215],[190,0.2096],[191,0.1952],[192,0.3898],[193,0.3472],[194,1.0542],[195,0.894],[196,0.3583],[197,0.1459],[198,0.1365],[199,0.1591],[200,0.1643],[201,0.1714],[202,0.1035],[203,0.11],[204,0.1054],[205,0.1061],[206,0.1031],[207,0.1082],[208,0.1011],[209,0.0937],[210,0.1073],[211,0.1111],[212,0.1058],[213,0.0982],[214,0.0897],[215,0.0834],[216,0.0817],[217,0.0777],[218,0.0779],[219,0.0754],[220,0.0758],[221,0.077],[222,0.0798],[223,0.0805],[224,0.0829],[225,0.0848],[227,0.0934],[228,0.096],[229,0.1043],[230,0.1147],[231,0.1339],[232,0.1426],[233,0.1966],[234,0.2401],[235,0.2806],[236,0.0061],[237,-0.4098],[238,-0.2574],[239,-0.6189],[240,-0.4158],[241,-0.0182],[242,0.002],[243,0.0087],[244,0.0349],[245,0.0638],[246,0.1133],[247,0.1088],[248,0.1328],[249,0.1741],[250,0.2009],[251,0.2149],[252,0.2226],[253,0.2397],[254,0.2584],[255,0.2578],[256,0.2559],[257,0.2336],[258,-0.0084],[259,-0.047],[260,-0.2814],[261,0.2948],[262,1.0927],[263,1.0352],[264,1.0959],[265,1.0857],[266,0.8233],[267,0.5253],[268,0.289],[269,0.0721],[270,0.234],[271,0.478],[272,0.4601],[273,0.4157],[274,0.3471],[275,0.3056],[276,0.3012],[277,0.3096],[278,0.2999],[279,0.295],[280,0.2812],[282,0.2875],[283,0.2843],[285,0.2838],[286,0.2996],[287,0.2968],[288,0.2987],[289,0.3151],[290,0.3223],[291,0.3146],[292,0.3093],[293,-0.0057],[294,0.1435],[295,0.4401],[296,0.4207],[297,0.4084],[298,0.303],[299,0.2424],[300,0.162],[301,0.1571],[302,0.1219],[303,0.1166],[304,0.0991],[305,0.0871],[306,0.0741],[307,0.0663],[308,0.0602],[309,0.0571],[310,0.0562],[311,0.0513],[312,0.0495],[313,0.0497],[314,0.0536],[315,0.0537],[316,0.0626],[317,0.0647],[318,0.0726],[319,0.0745],[320,0.0782],[321,0.0718],[322,0.0731],[323,0.0718],[324,0.0679],[325,0.0688],[326,0.0714],[327,0.0765],[328,0.0699],[330,0.0836],[331,0.0851],[332,0.0829],[333,0.0894],[334,0.0952],[335,0.0997],[336,0.1004],[337,0.0993],[338,0.1016],[339,0.1027],[340,0.1065],[341,0.1041],[342,0.1081],[343,0.1012],[344,0.1033],[345,0.1015],[346,0.1069],[347,0.1065],[348,0.1077],[349,0.1191],[350,0.1207],[351,0.1262],[352,0.1255],[353,0.1274],[355,0.1323],[356,0.1321],[358,0.1375],[359,0.1406],[360,0.1406]]
       }
       CTRL_R_eyeAim.Rotation.Z {
        type=MovieSceneScriptingFloatChannel
        default=0
        thinned_from=492
        keys [[-131,-0.0587],[-130,-0.0745],[-129,-0.0801],[-128,-0.0815],[-127,-0.0701],[-126,-0.0382],[-125,-0.0278],[-124,-0.0384],[-123,-0.0408],[-122,-0.0411],[-121,-0.0439],[-120,-0.0972],[-119,-0.1101],[-118,-0.0671],[-117,-0.0678],[-116,-0.0772],[-115,-0.071],[-114,-0.0602],[-113,-0.0733],[-112,-0.0671],[-111,-0.05],[-110,-0.0443],[-109,-0.0607],[-108,-0.0365],[-107,-0.0599],[-106,-0.0643],[-105,-0.0506],[-104,-0.0726],[-103,-0.0782],[-102,-0.0801],[-100,-0.0805],[-99,-0.0622],[-98,-0.0746],[-97,-0.0616],[-96,-0.088],[-95,-0.0987],[-94,-0.072],[-93,-0.0704],[-92,-0.0549],[-91,-0.0923],[-90,-0.0769],[-89,-0.0626],[-88,-0.0851],[-87,-0.0971],[-86,-0.0988],[-85,-0.0965],[-84,-0.1031],[-83,-0.1135],[-82,-0.1181],[-81,-0.1435],[-80,-0.1818],[-79,-0.1828],[-78,-0.1891],[-77,-0.1261],[-76,-0.076],[-75,-0.0512],[-74,-0.018],[-73,-0.0072],[-72,-0.0016],[-71,0.0223],[-70,-0.0203],[-69,0.0461],[-68,0.0179],[-67,0.0311],[-66,0.0352],[-65,0.0125],[-64,0.0245],[-63,0.012],[-62,0.0286],[-61,0.0395],[-60,0.0506],[-59,0.0315],[-58,0.0684],[-57,0.0704],[-56,0.0744],[-55,0.0634],[-54,0.0778],[-53,0.103],[-52,0.1074],[-51,0.0609],[-50,0.0666],[-49,0.0684],[-48,0.0408],[-47,0.0383],[-46,0.0322],[-45,0.0273],[-44,0.0213],[-43,-0.0034],[-42,-0.0421],[-41,-0.0685],[-40,-0.1056],[-39,-0.1262],[-38,-0.1237],[-37,-0.099],[-36,-0.0874],[-35,-0.0284],[-34,-0.0363],[-33,-0.0134],[-32,-0.0057],[-31,0.0067],[-30,0.0416],[-29,0.0409],[-28,0.0387],[-27,0.0377],[-26,0.0377],[-25,0.0575],[-24,0.0652],[-23,0.0505],[-22,-0.0693],[-21,-0.1027],[-20,-0.0621],[-19,-0.0484],[-18,-0.0423],[-17,-0.0157],[-16,-0.0052],[-15,-0.0024],[-14,-0.0291],[-13,-0.0362],[-12,-0.0356],[-11,0.0125],[-10,-0.0171],[-9,-0.032],[-8,-0.0103],[-7,0.0199],[-6,0.0082],[-5,0.0008],[-4,-0.0243],[-3,0.0204],[-2,0.0067],[-1,0.0016],[0,0.0002],[1,-0.0214],[2,-0.0106],[3,-0.0059],[4,-0.0282],[5,0.0114],[6,-0.0343],[7,-0.0379],[8,-0.0462],[9,-0.0676],[10,-0.0654],[11,-0.0476],[12,-0.0428],[13,-0.0514],[14,-0.0708],[15,-0.0593],[16,-0.0625],[17,-0.0522],[18,-0.0656],[19,-0.0499],[20,-0.0436],[21,-0.0413],[22,-0.0404],[23,-0.0476],[24,-0.0688],[25,-0.0763],[26,-0.0783],[27,-0.0668],[29,-0.0569],[30,-0.0634],[31,-0.0536],[32,-0.0439],[33,-0.0649],[34,-0.0486],[35,-0.0478],[36,-0.0463],[37,-0.0589],[38,-0.057],[39,-0.0657],[40,-0.056],[41,-0.0713],[42,-0.0659],[43,-0.0483],[44,-0.0421],[45,-0.0413],[46,-0.039],[47,-0.0154],[48,-0.0273],[49,-0.0109],[50,0.0194],[51,0.0001],[52,-0.0235],[53,-0.0146],[54,-0.0102],[55,-0.0043],[56,-0.019],[57,-0.0169],[58,-0.0254],[59,-0.0364],[60,-0.0376],[61,-0.0122],[62,-0.0128],[63,-0.0143],[64,-0.0377],[65,-0.0377],[66,-0.0393],[67,-0.0148],[68,-0.0051],[69,-0.0027],[70,-0.002],[71,-0.0019],[72,-0.0279],[73,-0.002],[75,-0.0342],[76,0.0138],[77,0.0308],[78,0.0277],[79,-0.0209],[80,-0.0341],[81,-0.0395],[82,-0.0153],[83,-0.0067],[84,-0.004],[85,-0.0021],[86,0.0237],[87,0.0331],[88,0.0351],[89,-0.0072],[90,0.0188],[91,0.0299],[92,0.0279],[93,0.0294],[94,0.0273],[95,0.0077],[96,0.0006],[97,0.0044],[98,0.0141],[99,0.0028],[100,-0.0005],[101,0.0158],[102,0.0236],[103,0.014],[104,0.0263],[105,0.0273],[106,0.0279],[107,0.0104],[108,0.0026],[109,0.0248],[110,0.0065],[111,0.026],[112,0.0075],[113,0.028],[114,0.0336],[115,0.0356],[116,0.0361],[117,0.0352],[118,0.0101],[119,0.002],[120,-0.0014],[121,-0.0028],[122,-0.0027],[124,-0.0021],[125,0.0211],[126,0.0307],[127,-0.0188],[128,0.0186],[129,0.0304],[130,0.0607],[131,0.0248],[132,0.0309],[133,0.0614],[134,0.0433],[135,0.0122],[136,0.0024],[137,0.0256],[138,0.0853],[139,0.051],[140,0.0161],[141,0.0232],[142,0.0102],[143,0.0039],[144,-0.0004],[145,0.0475],[146,-0.0153],[147,-0.0059],[148,0.0258],[149,0.0333],[150,0.036],[151,0.0105],[152,0.0024],[153,-0.0011],[154,-0.0017],[155,0.0236],[156,0.0326],[157,0.0577],[158,0.0704],[159,0.0747],[160,0.0488],[161,0.0413],[162,0.0388],[163,0.0377],[164,0.0132],[165,0.0263],[166,0.0347],[168,0.0387],[169,0.0561],[170,0.0442],[171,0.0392],[172,0.0187],[173,0.0047],[174,0.0239],[175,0.0087],[176,0.0286],[177,0.0238],[178,0.0058],[179,0.0006],[180,-0.0003],[181,-0.0316],[182,-0.0241],[183,-0.0058],[184,0.0172],[185,0.0123],[186,0.0032],[187,0],[189,-0.0008],[190,-0.0037],[191,-0.0261],[192,-0.0704],[193,-0.1016],[194,-0.1801],[195,-0.1731],[196,-0.1594],[197,-0.1562],[198,-0.1186],[199,-0.1075],[200,-0.1112],[201,-0.0904],[202,-0.1053],[203,-0.0866],[204,-0.0539],[205,-0.1001],[206,-0.0988],[207,-0.0618],[208,-0.0734],[209,-0.048],[210,-0.0167],[211,-0.0335],[212,-0.0357],[213,-0.0403],[214,-0.0656],[215,-0.0301],[216,-0.0323],[217,-0.0375],[218,-0.0385],[219,-0.0126],[220,-0.0053],[221,-0.0287],[222,-0.0096],[223,-0.0037],[224,0.0225],[225,0.0329],[226,0.0094],[227,0.0028],[228,-0.0252],[229,-0.0087],[230,-0.0036],[231,-0.0019],[232,-0.0264],[233,-0.0375],[234,-0.04],[235,-0.0718],[236,-0.1416],[237,-0.1784],[238,-0.1918],[239,-0.2417],[240,-0.1914],[241,-0.1198],[242,-0.1096],[243,-0.1168],[244,-0.1175],[245,-0.1171],[246,-0.0992],[247,-0.134],[248,-0.1249],[249,-0.1212],[250,-0.1191],[251,-0.0949],[252,-0.1131],[253,-0.1175],[254,-0.1452],[255,-0.1279],[256,-0.0983],[257,-0.1175],[258,-0.1134],[259,-0.117],[260,-0.173],[261,-0.2397],[262,-0.2494],[263,-0.2579],[264,-0.262],[265,-0.2642],[266,-0.2573],[267,-0.2348],[268,-0.2089],[269,-0.1295],[270,-0.1149],[271,-0.1202],[272,-0.0686],[273,-0.0774],[274,-0.0873],[275,-0.1008],[276,-0.0879],[277,-0.0832],[278,-0.0812],[279,-0.0674],[280,-0.0657],[281,-0.0946],[282,-0.0914],[283,-0.0837],[284,-0.083],[285,-0.0814],[286,-0.0594],[287,-0.0654],[288,-0.0749],[289,-0.0799],[291,-0.0818],[292,-0.123],[293,-0.1767],[294,-0.1729],[295,-0.1227],[296,-0.1178],[297,-0.1171],[298,-0.0794],[299,-0.0591],[300,-0.0523],[301,-0.0495],[302,-0.0151],[303,-0.004],[304,0.0274],[305,0.0347],[306,0.0374],[307,0.038],[308,0.0555],[309,0.0516],[310,0.0454],[311,0.0064],[312,0.0033],[313,0.0006],[314,0.0031],[315,0.051],[316,0.0219],[317,0.0319],[318,0.0099],[319,0.0027],[320,0.0002],[321,-0.0003],[322,0.0234],[323,0.007],[324,0.0016],[325,-0.0001],[326,-0.0007],[327,-0.001],[328,-0.0009],[329,-0.0033],[330,-0.0516],[331,-0.0209],[332,-0.0071],[333,-0.0278],[334,-0.0102],[335,-0.0053],[336,-0.0023],[337,-0.0016],[338,0.0271],[339,0.0094],[340,0.0018],[341,0],[342,-0.0282],[343,-0.0116],[344,-0.004],[345,-0.0022],[346,-0.0029],[347,-0.0272],[348,-0.0104],[349,-0.0264],[350,-0.0132],[351,-0.0049],[352,-0.0028],[353,-0.0333],[354,-0.0383],[355,-0.0401],[356,-0.0403],[357,-0.0123],[358,-0.0053],[359,-0.0297],[360,-0.0297]]
       }
       CTRL_R_eyeAim.Scale.X {type=MovieSceneScriptingFloatChannel, default=1, const [1,492]}
       CTRL_R_eyeAim.Scale.Y {type=MovieSceneScriptingFloatChannel, default=1, const [1,492]}
       CTRL_R_eyeAim.Scale.Z {type=MovieSceneScriptingFloatChannel, default=1, const [1,492]}
       CTRL_convergenceSwitch {type=MovieSceneScriptingFloatChannel, default=0, const [0,492]}
       CTRL_Master.Location.X {type=MovieSceneScriptingFloatChannel, default=0, const [0,492]}
       CTRL_Master.Location.Y {type=MovieSceneScriptingFloatChannel, default=0, const [0,492]}
       CTRL_Master.Location.Z {type=MovieSceneScriptingFloatChannel, default=0, const [0,492]}
       CTRL_Master.Rotation.X {type=MovieSceneScriptingFloatChannel, default=0, const [0,492]}
       CTRL_Master.Rotation.Y {type=MovieSceneScriptingFloatChannel, default=0, const [0,492]}
       CTRL_Master.Rotation.Z {type=MovieSceneScriptingFloatChannel, default=0, const [0,492]}
       CTRL_Master.Scale.X {type=MovieSceneScriptingFloatChannel, default=1, const [1,492]}
       CTRL_Master.Scale.Y {type=MovieSceneScriptingFloatChannel, default=1, const [1,492]}
       CTRL_Master.Scale.Z {type=MovieSceneScriptingFloatChannel, default=1, const [1,492]}
       CTRL_C_jaw.X {
        type=MovieSceneScriptingFloatChannel
        default=0
        thinned_from=492
        keys [[-131,0.0108],[-130,0.0175],[-129,0.0269],[-128,0.0306],[-127,0.0317],[-126,0.0292],[-125,0.0223],[-124,0.0172],[-123,0.0151],[-122,0.0147],[-121,0.0124],[-120,0.0052],[-119,0.0053],[-118,-0.0028],[-117,-0.004],[-116,-0.0119],[-114,-0.0241],[-113,-0.0253],[-112,-0.0234],[-111,-0.02],[-110,-0.0193],[-109,-0.0188],[-108,-0.0202],[-107,-0.0222],[-106,-0.0225],[-105,-0.0238],[-104,-0.0191],[-103,-0.0129],[-102,-0.0106],[-101,-0.008],[-100,-0.0043],[-99,-0.0029],[-98,0.0009],[-97,0.0029],[-96,0.0066],[-95,0.0075],[-94,-0.0022],[-93,-0.0023],[-92,0.0019],[-91,-0.0021],[-90,-0.0035],[-89,-0.0021],[-88,-0.0014],[-87,0.0014],[-86,0.0007],[-85,0.0012],[-84,0.0119],[-83,0.0185],[-82,0.0236],[-81,0.0307],[-80,0.0415],[-79,0.0423],[-78,0.0491],[-77,0.0439],[-76,0.0355],[-75,0.0327],[-74,0.0359],[-73,0.0322],[-72,0.0377],[-71,0.0469],[-69,0.0572],[-68,0.0595],[-67,0.056],[-66,0.049],[-65,0.0467],[-64,0.0487],[-63,0.0482],[-62,0.0487],[-61,0.0518],[-60,0.0556],[-59,0.0523],[-58,0.04],[-57,0.0372],[-56,0.0293],[-55,0.0216],[-54,0.0196],[-53,0.0206],[-52,0.0208],[-51,0.0137],[-50,0.0118],[-49,0.0117],[-48,0.0122],[-47,0.0108],[-46,0.0101],[-45,0.0088],[-44,0.0089],[-43,0.0093],[-41,0.008],[-40,0.012],[-39,0.0207],[-38,0.0233],[-37,0.0295],[-36,0.0403],[-35,0.0569],[-34,0.0624],[-33,0.0611],[-32,0.0505],[-31,0.0451],[-30,0.0192],[-29,0.0177],[-28,0.0143],[-27,0.0125],[-26,0.0096],[-25,0.0059],[-24,0.0047],[-23,-0.0043],[-22,-0.01],[-21,-0.0078],[-20,-0.0036],[-19,-0.0018],[-18,0.0004],[-17,0.0053],[-16,0.0064],[-15,0.0083],[-14,0.0087],[-13,0.0068],[-12,0.008],[-11,0.0102],[-10,0.0087],[-9,0.0086],[-8,0.0072],[-7,0.0063],[-6,0.0097],[-5,0.0107],[-4,0.0098],[-3,0.0098],[-2,0.0102],[-1,0.0101],[1,0.0055],[2,0.0049],[3,0.0068],[4,0.0068],[5,0.0038],[6,-0.0047],[8,-0.0066],[10,-0.0136],[11,-0.0132],[12,-0.0131],[14,-0.0116],[15,-0.0111],[17,-0.0079],[19,-0.0018],[20,-0.0003],[21,-0.0027],[22,-0.0067],[23,-0.0071],[24,-0.0079],[25,-0.0082],[27,-0.0078],[28,-0.0071],[30,-0.0072],[31,-0.0081],[32,-0.0099],[33,-0.0097],[34,-0.0088],[35,-0.0085],[36,-0.006],[37,-0.0057],[38,-0.0058],[39,-0.0054],[40,-0.0059],[41,-0.0049],[42,-0.0044],[43,-0.0034],[44,-0.0032],[45,-0.0034],[46,-0.0084],[47,-0.0107],[48,-0.0082],[49,-0.0092],[50,-0.0074],[52,-0.0001],[53,0.0024],[54,0.0027],[56,0.0047],[57,0.0051],[58,0.003],[60,0.0039],[61,0.0069],[62,0.0083],[63,0.0081],[64,0.0049],[65,0.0049],[66,0.0069],[67,0.0129],[68,0.0157],[69,0.0175],[70,0.0171],[72,0.0171],[73,0.0165],[74,0.0165],[75,0.015],[76,0.0152],[77,0.0166],[78,0.0167],[80,0.0214],[83,0.0245],[84,0.0263],[85,0.0263],[86,0.023],[87,0.0225],[88,0.0244],[89,0.0232],[90,0.0225],[91,0.0231],[92,0.0249],[93,0.0257],[94,0.0275],[95,0.0272],[96,0.0257],[101,0.0244],[102,0.0244],[103,0.0252],[104,0.0272],[105,0.0273],[106,0.0272],[107,0.0269],[108,0.0268],[110,0.024],[111,0.0214],[112,0.0182],[113,0.0156],[114,0.0156],[115,0.0135],[117,0.0043],[118,0.001],[119,-0.0007],[120,0.0004],[121,0.0043],[122,0.0022],[123,-0.0007],[124,0.0017],[125,0.0038],[127,0.0049],[128,0.002],[129,0.0034],[130,0.0078],[131,0.0078],[132,0.0084],[133,0.0107],[134,0.0139],[135,0.0185],[136,0.014],[137,0.0107],[138,0.0109],[139,0.0064],[140,0.0068],[141,0.0064],[142,0.0068],[143,0.0097],[144,0.0112],[146,0.0064],[147,0.0064],[148,0.0078],[149,0.0084],[150,0.0098],[151,0.0101],[152,0.0078],[153,0.0053],[154,0.0039],[156,0.0068],[157,0.0068],[159,0.0078],[160,0.0071],[161,0.004],[162,0.0051],[163,0.0087],[165,0.0073],[166,0.0082],[167,0.0059],[168,0.0054],[169,0.0057],[170,-0.0024],[171,-0.0086],[172,-0.0079],[173,-0.0079],[174,-0.005],[175,-0.0016],[176,0.0028],[177,0.0038],[178,0.0061],[179,0.0042],[181,0.0038],[182,0.0017],[184,0.0028],[185,0.0029],[187,-0.0003],[188,-0.0005],[189,-0.001],[190,0],[191,-0.0007],[192,-0.0081],[193,-0.0157],[194,-0.0112],[195,-0.0079],[196,-0.0058],[197,-0.0088],[198,-0.0084],[199,-0.0018],[200,0.0004],[201,0.0009],[202,-0.0006],[203,-0.0034],[205,-0.0057],[206,-0.0051],[207,-0.0019],[208,0.0002],[209,-0.0005],[210,-0.001],[211,-0.0007],[212,0.0002],[213,0.0034],[214,0.0054],[216,0.0077],[217,0.0087],[218,0.0102],[219,0.0093],[220,0.0102],[221,0.0088],[222,0.0093],[223,0.0112],[224,0.0127],[225,0.0137],[226,0.0127],[227,0.0122],[228,0.0123],[229,0.0137],[230,0.0123],[232,0.0041],[233,-0.0038],[235,-0.0178],[236,-0.0172],[237,-0.0198],[238,-0.0185],[239,-0.013],[240,-0.0039],[241,0.0002],[242,0.0005],[243,-0.0012],[244,0.0029],[245,0.0076],[246,0.0082],[247,0.0194],[248,0.0318],[249,0.0346],[250,0.0312],[251,0.0293],[252,0.0177],[253,0.008],[254,0.0081],[255,0.0032],[256,-0.0022],[257,-0.0068],[258,-0.0124],[259,-0.0124],[260,-0.0085],[261,0.0027],[262,0.0127],[263,0.0147],[264,0.0163],[265,0.0189],[266,0.0199],[267,0.0179],[268,0.017],[269,0.0215],[270,0.0225],[271,0.0193],[272,0.0156],[273,0.0131],[274,0.0038],[275,-0.0026],[276,-0.0016],[277,0.0014],[278,-0.0007],[279,-0.0015],[280,0.0003],[282,0.0009],[283,0.0014],[284,0.0015],[285,0.0031],[286,0.0089],[287,0.0093],[288,0.0089],[289,0.0089],[290,0.0093],[291,0.0091],[292,0.0107],[293,0.0149],[294,0.0171],[295,0.0199],[296,0.0201],[297,0.0197],[298,0.011],[299,0.0068],[300,0.0073],[301,0.0081],[302,0.0142],[303,0.0139],[304,0.0134],[305,0.0142],[307,0.0178],[308,0.0208],[309,0.0233],[310,0.024],[311,0.0268],[312,0.0273],[314,0.0274],[316,0.0285],[317,0.0288],[318,0.0279],[319,0.0288],[320,0.0288],[321,0.0269],[322,0.0269],[323,0.0273],[324,0.0269],[325,0.0279],[326,0.0293],[327,0.0313],[328,0.0332],[329,0.0331],[330,0.0305],[331,0.0288],[332,0.0293],[333,0.028],[334,0.0279],[335,0.0283],[336,0.0278],[337,0.0278],[338,0.0273],[340,0.0254],[342,0.0283],[344,0.0273],[345,0.0273],[347,0.0264],[348,0.0274],[349,0.0289],[350,0.0293],[351,0.0278],[352,0.0283],[353,0.0283],[354,0.0281],[356,0.0283],[357,0.0279],[358,0.0288],[359,0.0304],[360,0.0304]]
       }
       CTRL_C_jaw.Y {
        type=MovieSceneScriptingFloatChannel
        default=0
        thinned_from=492
        keys [[-131,0.2531],[-130,0.2537],[-129,0.2611],[-128,0.2607],[-127,0.2664],[-126,0.2727],[-125,0.2808],[-124,0.2928],[-123,0.2986],[-122,0.2995],[-121,0.3014],[-120,0.3077],[-119,0.3088],[-118,0.3128],[-117,0.3136],[-116,0.3191],[-114,0.322],[-112,0.333],[-111,0.3339],[-110,0.3356],[-109,0.3399],[-108,0.3385],[-107,0.332],[-106,0.3309],[-105,0.3273],[-103,0.3316],[-102,0.3293],[-101,0.3279],[-100,0.3299],[-99,0.331],[-98,0.329],[-97,0.3176],[-96,0.3131],[-95,0.3123],[-94,0.301],[-93,0.3003],[-92,0.2975],[-91,0.2919],[-90,0.2924],[-89,0.2968],[-88,0.3072],[-87,0.3187],[-86,0.3232],[-85,0.3291],[-84,0.3407],[-83,0.3517],[-82,0.3612],[-81,0.3567],[-80,0.3451],[-79,0.3438],[-78,0.3313],[-77,0.2955],[-76,0.2664],[-75,0.2506],[-74,0.2368],[-73,0.2252],[-72,0.2219],[-71,0.2216],[-70,0.2241],[-69,0.2233],[-68,0.2206],[-67,0.2191],[-66,0.2114],[-65,0.1972],[-64,0.1906],[-63,0.1902],[-62,0.181],[-61,0.1611],[-60,0.1345],[-59,0.0972],[-58,0.0635],[-57,0.0601],[-56,0.0524],[-55,0.0483],[-54,0.0462],[-53,0.0448],[-52,0.0442],[-51,0.0456],[-50,0.0477],[-49,0.048],[-48,0.0509],[-47,0.0545],[-46,0.0562],[-45,0.0694],[-44,0.0759],[-43,0.1023],[-42,0.1388],[-41,0.168],[-40,0.2112],[-39,0.2482],[-38,0.2565],[-37,0.2463],[-36,0.2396],[-35,0.2291],[-34,0.2083],[-33,0.1662],[-32,0.1217],[-31,0.1104],[-30,0.0584],[-29,0.0568],[-28,0.0534],[-27,0.0497],[-26,0.0483],[-24,0.0482],[-23,0.0523],[-22,0.0605],[-21,0.0683],[-20,0.0717],[-19,0.0697],[-18,0.0733],[-17,0.0717],[-16,0.0662],[-15,0.0662],[-14,0.0686],[-13,0.0686],[-12,0.0646],[-11,0.0615],[-10,0.0605],[-8,0.0574],[-6,0.0546],[-5,0.0541],[-4,0.0541],[-3,0.0536],[-2,0.0518],[-1,0.0496],[0,0.05],[1,0.0494],[2,0.0482],[3,0.0486],[4,0.0493],[6,0.052],[7,0.0513],[9,0.0512],[10,0.0506],[11,0.0498],[12,0.0494],[13,0.0493],[14,0.0503],[15,0.051],[16,0.0521],[17,0.0529],[18,0.051],[19,0.0504],[20,0.0517],[21,0.0516],[22,0.0509],[23,0.0509],[24,0.0513],[25,0.0521],[27,0.0504],[28,0.0504],[29,0.0511],[30,0.0513],[31,0.0506],[32,0.0504],[33,0.0505],[34,0.0497],[35,0.0496],[36,0.0493],[37,0.0495],[39,0.0485],[40,0.049],[41,0.0492],[42,0.0479],[43,0.0478],[44,0.0484],[45,0.0485],[46,0.0472],[47,0.0493],[48,0.0547],[49,0.0569],[50,0.06],[51,0.0614],[52,0.0588],[53,0.055],[54,0.0548],[55,0.0553],[56,0.058],[57,0.0586],[58,0.0587],[59,0.0636],[60,0.0648],[61,0.067],[62,0.0663],[63,0.0662],[64,0.0652],[65,0.0652],[66,0.0633],[67,0.0621],[68,0.0637],[69,0.0638],[70,0.0612],[71,0.06],[72,0.0579],[73,0.0577],[74,0.0555],[75,0.0553],[76,0.0543],[77,0.0519],[78,0.0518],[79,0.0511],[80,0.0499],[82,0.0493],[83,0.0475],[84,0.0484],[85,0.0485],[86,0.0466],[87,0.0469],[88,0.0476],[90,0.0471],[92,0.0472],[93,0.047],[94,0.0464],[95,0.0466],[96,0.0473],[97,0.0472],[98,0.0474],[99,0.048],[100,0.0481],[101,0.0494],[102,0.0499],[103,0.0466],[104,0.0476],[105,0.0477],[106,0.0498],[108,0.0496],[109,0.0506],[110,0.0526],[112,0.052],[113,0.0551],[114,0.0549],[115,0.0534],[116,0.0543],[117,0.0542],[118,0.056],[119,0.0574],[120,0.0576],[121,0.0567],[122,0.0552],[123,0.0559],[124,0.0555],[126,0.0513],[127,0.0512],[128,0.0515],[129,0.0501],[130,0.0482],[133,0.0457],[134,0.0468],[135,0.0488],[136,0.0525],[137,0.0536],[138,0.0471],[139,0.0431],[141,0.0431],[142,0.0446],[144,0.0512],[145,0.0528],[146,0.0559],[147,0.0562],[148,0.0547],[149,0.0541],[150,0.0538],[151,0.0526],[152,0.0502],[153,0.0499],[154,0.0512],[155,0.0527],[156,0.0522],[157,0.0486],[158,0.0472],[159,0.0487],[160,0.0491],[161,0.0489],[162,0.0478],[163,0.0486],[164,0.05],[165,0.0486],[166,0.0483],[167,0.0492],[168,0.0535],[169,0.0623],[170,0.0643],[171,0.0673],[172,0.0698],[173,0.0683],[174,0.0744],[175,0.0841],[176,0.0857],[177,0.0855],[178,0.0845],[179,0.085],[180,0.0873],[181,0.0941],[182,0.0956],[183,0.0946],[184,0.0898],[185,0.0856],[187,0.0914],[188,0.0918],[189,0.0909],[190,0.0925],[191,0.0924],[192,0.0913],[193,0.0859],[194,0.0853],[195,0.085],[196,0.093],[197,0.1135],[198,0.1271],[199,0.1314],[200,0.1325],[201,0.1504],[202,0.1422],[203,0.1227],[204,0.128],[205,0.1271],[206,0.1241],[207,0.1251],[208,0.1255],[209,0.1198],[210,0.1131],[211,0.1037],[212,0.1032],[213,0.1032],[214,0.0995],[215,0.098],[216,0.099],[217,0.0991],[218,0.0983],[220,0.0989],[221,0.0982],[222,0.0952],[224,0.0994],[225,0.0975],[226,0.0976],[227,0.0995],[228,0.0993],[229,0.1005],[230,0.1078],[231,0.1142],[232,0.1191],[233,0.1383],[234,0.1529],[235,0.1555],[236,0.1356],[237,0.1162],[238,0.1099],[239,0.1201],[240,0.1201],[241,0.1261],[242,0.1282],[243,0.1426],[244,0.1515],[245,0.1671],[246,0.1851],[247,0.2008],[248,0.2173],[249,0.227],[250,0.2239],[251,0.2252],[252,0.2328],[253,0.2387],[254,0.2432],[255,0.2337],[256,0.232],[257,0.2533],[258,0.2429],[259,0.243],[260,0.2419],[261,0.2351],[262,0.2225],[263,0.2033],[264,0.2088],[265,0.2155],[266,0.216],[267,0.2186],[268,0.2181],[269,0.2344],[270,0.2379],[271,0.2577],[272,0.2621],[273,0.2707],[274,0.2984],[275,0.3127],[276,0.3178],[277,0.316],[278,0.318],[280,0.3174],[281,0.3201],[283,0.3179],[284,0.3179],[285,0.3206],[286,0.3148],[288,0.3117],[289,0.3093],[291,0.303],[292,0.2992],[293,0.2835],[294,0.2729],[295,0.2718],[296,0.2546],[297,0.2498],[298,0.2032],[299,0.1729],[300,0.1335],[301,0.1297],[302,0.0964],[303,0.0922],[304,0.0787],[305,0.0714],[306,0.0685],[307,0.068],[308,0.067],[309,0.0665],[310,0.0667],[311,0.0706],[312,0.0706],[313,0.0679],[314,0.0642],[315,0.0622],[316,0.0626],[318,0.0623],[319,0.0624],[320,0.064],[321,0.0653],[322,0.0647],[323,0.0646],[324,0.0655],[325,0.0656],[326,0.0662],[327,0.0675],[328,0.0675],[329,0.0681],[330,0.0712],[331,0.0727],[332,0.0698],[333,0.0689],[335,0.0693],[336,0.0692],[337,0.0665],[338,0.0668],[339,0.0683],[340,0.0693],[342,0.0697],[343,0.0695],[344,0.0678],[345,0.068],[346,0.0702],[347,0.0719],[348,0.0713],[349,0.0698],[350,0.0699],[351,0.0704],[352,0.0695],[353,0.0714],[354,0.0711],[355,0.0697],[356,0.0695],[357,0.0689],[358,0.0678],[360,0.0678]]
       }
       CTRL_C_jaw_openExtreme {
        type=MovieSceneScriptingFloatChannel
        default=0
        thinned_from=492
        keys [[-131,0.002],[-130,0.0007],[-129,0.0009],[-128,0.0002],[-126,0],[-124,0.0005],[-123,0.0001],[-121,0],[-120,0.0007],[-119,0],[-118,0.0005],[-117,0.0005],[-115,0.001],[-113,0.0006],[-111,0.0007],[-110,0.0002],[-109,0.0002],[-108,0.001],[-107,0.0012],[-106,0.0012],[-105,0.001],[-104,0.0011],[-103,0.0006],[-102,0.001],[-101,0.0007],[-100,0.0003],[-99,0.0007],[-98,0.0009],[-97,0.0009],[-95,0.0012],[-94,0.0008],[-91,0.0006],[-90,0.001],[-89,0.0011],[-87,0.0019],[-86,0.0021],[-85,0.002],[-83,0.0024],[-79,0.0027],[-77,0.0021],[-76,0.0024],[-75,0.002],[-73,0.0024],[-68,0.0024],[-67,0.0029],[-66,0.0025],[-64,0.0026],[-63,0.0025],[-60,0.0026],[-58,0.0022],[-56,0.0023],[-55,0.0021],[-54,0.0021],[-52,0.0024],[-51,0.0021],[-43,0.0021],[-42,0.002],[-38,0.0024],[-37,0.0022],[-36,0.0024],[-35,0.0029],[-34,0.0025],[-33,0.0025],[-32,0.0027],[-31,0.0026],[-30,0.002],[-23,0.0017],[-22,0.0022],[-21,0.002],[-20,0.0022],[-19,0.0022],[-18,0.0024],[-17,0.0024],[-16,0.0022],[-14,0.0022],[-13,0.0024],[-12,0.0022],[-11,0.0024],[-10,0.0024],[-9,0.0022],[-8,0.0024],[-7,0.0022],[0,0.0022],[1,0.0024],[3,0.002],[4,0.0022],[5,0.0022],[6,0.0024],[9,0.0023],[10,0.0024],[14,0.0022],[16,0.0024],[19,0.0022],[20,0.0024],[21,0.0024],[22,0.002],[29,0.0023],[32,0.0022],[33,0.002],[39,0.0022],[40,0.0024],[45,0.0022],[46,0.0024],[47,0.0024],[49,0.002],[50,0.0022],[52,0.0022],[53,0.002],[54,0.0022],[56,0.0019],[58,0.0005],[59,0.001],[61,0.0012],[62,0.0017],[63,0.0017],[64,0.0015],[65,0.0015],[66,0.0017],[67,0.0012],[68,0.0012],[69,0.0015],[70,0.0012],[72,0.0012],[73,0.001],[75,0.0019],[76,0.0015],[77,0.0017],[78,0.0017],[79,0.0015],[80,0.001],[81,0.0015],[82,0.0012],[83,0.0012],[84,0.0007],[85,0.0012],[86,0.001],[87,0.0012],[88,0.0005],[89,0.0007],[91,0.0005],[92,0.0007],[94,0.0006],[95,0.0007],[97,0.0005],[98,0.0006],[99,0.001],[100,0.0011],[102,0.001],[103,0.0007],[104,0.001],[105,0.001],[106,0.0004],[107,0.001],[108,0.0005],[109,0.0007],[110,0.0007],[111,0.001],[112,0.0007],[113,0.0007],[114,0.0005],[115,0.001],[116,0.001],[117,0.0005],[118,0.0007],[119,0.0005],[120,0.001],[121,0.001],[123,0.0005],[124,0.0007],[125,0],[127,0.0005],[128,0],[129,0.0005],[130,0.0005],[131,0],[132,0],[134,0.001],[136,0.0008],[137,0.0022],[138,0.0018],[139,0],[140,0.0014],[141,0.0011],[142,0.0012],[143,0.0016],[145,0.001],[146,0.001],[147,0.0014],[148,0.0007],[149,0.0007],[151,0.0012],[152,0.001],[153,0.0014],[154,0.001],[155,0.0012],[156,0.0005],[157,0.001],[158,0.0007],[159,0.0014],[160,0.0007],[162,0.0014],[164,0.001],[165,0.0012],[166,0.001],[167,0.0015],[168,0.001],[169,0.0017],[170,0.0017],[171,0.0021],[172,0.0018],[173,0.0009],[174,0.0003],[175,0.0012],[178,0.0013],[179,0.0004],[181,0.001],[183,0.0005],[185,0.0007],[186,0.0005],[188,0.0007],[190,0.0003],[192,0.0008],[193,0.002],[194,0.0027],[195,0.0025],[197,0.0013],[198,0.0021],[199,0.0022],[201,0.0018],[202,0.0028],[203,0.0022],[204,0.002],[205,0.0021],[206,0.0017],[208,0.0017],[210,0.0022],[212,0.0021],[213,0.0016],[215,0.0017],[216,0.0019],[219,0.0017],[220,0.0022],[221,0.0017],[222,0.0015],[223,0.0019],[224,0.0017],[229,0.0017],[230,0.0019],[232,0.0015],[233,0.0017],[234,0.0015],[235,0.0003],[236,0.002],[237,0.0021],[238,0.002],[239,0.0009],[241,0.0015],[242,0.0015],[243,0.0005],[244,0.0005],[245,0.0009],[246,0.0005],[247,0.0006],[248,0],[249,0.0007],[250,0],[251,0.0007],[253,0.0002],[254,0.001],[255,0.0007],[256,0],[258,0],[259,0.0002],[260,0.0017],[261,0.0023],[262,0.0019],[263,0.0024],[265,0.0029],[267,0.0024],[268,0.0026],[269,0.0019],[270,0.0017],[271,0.0019],[272,0.0015],[273,0.0007],[274,0.0002],[276,0.0002],[278,0.0004],[279,0.0001],[281,0.0005],[282,0.0003],[283,0.0007],[284,0.0007],[286,0],[288,0.0005],[292,0.0002],[293,0.0005],[295,0.0007],[296,0.0014],[297,0.0014],[298,0.0009],[300,0.0019],[304,0.0017],[307,0.0022],[308,0.0019],[310,0.0019],[311,0.0017],[312,0.0017],[313,0.0015],[314,0.0017],[316,0.0018],[317,0.0015],[318,0.0015],[320,0.0019],[321,0.0019],[322,0.0012],[323,0.0012],[324,0.0019],[325,0.0017],[326,0.0017],[327,0.0015],[328,0.0019],[330,0.0017],[331,0.002],[332,0.002],[333,0.0017],[337,0.0017],[338,0.0019],[339,0.0017],[340,0.0017],[341,0.0015],[342,0.0015],[343,0.0019],[344,0.0015],[345,0.0015],[346,0.0017],[347,0.0015],[348,0.0017],[350,0.0015],[351,0.0019],[352,0.0015],[353,0.0018],[354,0.0017],[356,0.002],[357,0.0017],[358,0.0017],[359,0.0015],[360,0.0015]]
       }
       CTRL_C_jaw_fwdBack {
        type=MovieSceneScriptingFloatChannel
        default=0
        thinned_from=492
        keys [[-131,0.1579],[-130,0.1642],[-129,0.1637],[-128,0.1649],[-127,0.1671],[-126,0.1677],[-125,0.1637],[-124,0.168],[-123,0.1697],[-122,0.1697],[-121,0.1713],[-120,0.1703],[-119,0.1707],[-118,0.1694],[-117,0.1697],[-116,0.1718],[-115,0.1703],[-114,0.1644],[-113,0.1612],[-112,0.161],[-111,0.1598],[-110,0.1605],[-109,0.1617],[-108,0.1643],[-107,0.1613],[-105,0.1602],[-104,0.1681],[-103,0.1744],[-102,0.173],[-101,0.1699],[-100,0.1751],[-99,0.1744],[-98,0.173],[-97,0.1684],[-96,0.1667],[-95,0.1663],[-94,0.1616],[-93,0.162],[-92,0.1645],[-91,0.1628],[-90,0.1642],[-89,0.1695],[-87,0.195],[-86,0.1978],[-85,0.1994],[-84,0.2029],[-83,0.1984],[-82,0.1928],[-81,0.1891],[-80,0.1722],[-79,0.1699],[-78,0.1492],[-77,0.1143],[-76,0.0907],[-75,0.0444],[-74,0.0284],[-73,0.0283],[-72,0.0174],[-71,0.0241],[-70,0.0118],[-69,0.0032],[-68,0.004],[-67,-0.0011],[-66,-0.0044],[-65,-0.0101],[-64,-0.0086],[-63,-0.0107],[-62,-0.0168],[-61,-0.0192],[-60,-0.025],[-59,-0.0252],[-58,-0.0128],[-57,-0.0065],[-56,0.0083],[-55,0.0179],[-54,0.024],[-53,0.0277],[-52,0.0293],[-51,0.0407],[-50,0.0448],[-49,0.0451],[-48,0.0588],[-46,0.0607],[-45,0.0714],[-44,0.0739],[-43,0.0838],[-42,0.0968],[-41,0.1066],[-40,0.1252],[-39,0.1348],[-38,0.1351],[-37,0.0808],[-36,0.0557],[-35,0.0321],[-34,0.0053],[-33,-0.0107],[-32,0.0013],[-31,0.0067],[-30,0.036],[-29,0.0391],[-28,0.0462],[-27,0.0483],[-26,0.0481],[-25,0.0419],[-24,0.0391],[-23,0.0242],[-22,0.0192],[-21,0.0182],[-20,0.0161],[-19,0.0236],[-18,0.037],[-17,0.0442],[-16,0.0556],[-15,0.0631],[-14,0.07],[-13,0.0717],[-12,0.0732],[-11,0.0743],[-10,0.0742],[-9,0.0722],[-8,0.0729],[-7,0.075],[-6,0.0743],[-5,0.0731],[-4,0.0745],[-3,0.0766],[-2,0.0754],[-1,0.0768],[0,0.078],[1,0.0768],[2,0.0772],[3,0.0758],[4,0.0755],[5,0.0764],[6,0.0803],[7,0.0836],[8,0.084],[9,0.0836],[10,0.0813],[11,0.0823],[12,0.0811],[14,0.0797],[16,0.0807],[17,0.0809],[18,0.0806],[19,0.0808],[20,0.0827],[21,0.0855],[22,0.0847],[23,0.0845],[24,0.0834],[25,0.0821],[26,0.0804],[27,0.0792],[28,0.0795],[29,0.0781],[30,0.0788],[31,0.0771],[32,0.0747],[33,0.0771],[34,0.0772],[35,0.077],[36,0.0748],[37,0.075],[38,0.078],[39,0.0788],[40,0.0754],[41,0.0759],[42,0.0767],[43,0.075],[44,0.074],[45,0.0743],[46,0.0655],[47,0.0622],[48,0.0527],[49,0.0483],[50,0.0437],[51,0.0447],[52,0.044],[53,0.0454],[54,0.0465],[55,0.051],[56,0.0557],[57,0.0583],[58,0.0616],[59,0.0656],[60,0.0663],[61,0.0681],[62,0.0707],[63,0.071],[64,0.0746],[65,0.0746],[66,0.0732],[67,0.0723],[68,0.0733],[69,0.0712],[70,0.0701],[71,0.066],[73,0.0663],[74,0.0676],[75,0.0674],[76,0.0687],[77,0.0651],[78,0.065],[79,0.0637],[80,0.0641],[81,0.0658],[82,0.0642],[83,0.0615],[84,0.0606],[86,0.0626],[87,0.0604],[88,0.0614],[89,0.0606],[90,0.061],[91,0.0621],[92,0.062],[93,0.0626],[95,0.066],[96,0.0669],[97,0.067],[98,0.0675],[99,0.0669],[100,0.068],[101,0.068],[102,0.0675],[103,0.0664],[104,0.0685],[105,0.0687],[106,0.0695],[107,0.069],[108,0.0715],[109,0.0728],[110,0.073],[111,0.0727],[112,0.0735],[113,0.0734],[114,0.0743],[115,0.0736],[116,0.0694],[118,0.0674],[119,0.062],[120,0.06],[121,0.0608],[122,0.0612],[123,0.0593],[124,0.0594],[125,0.056],[126,0.0543],[127,0.054],[128,0.0564],[129,0.0555],[130,0.0507],[131,0.0481],[132,0.0487],[133,0.0478],[134,0.05],[135,0.0545],[136,0.0596],[137,0.058],[138,0.0563],[139,0.055],[141,0.0557],[142,0.0594],[143,0.0616],[144,0.0646],[145,0.0645],[146,0.0635],[147,0.0647],[148,0.0629],[149,0.0675],[150,0.0707],[151,0.0709],[152,0.0692],[153,0.0657],[155,0.0682],[156,0.0683],[157,0.0675],[158,0.0664],[159,0.0637],[160,0.0656],[161,0.0661],[163,0.0687],[164,0.0695],[165,0.0691],[166,0.0665],[167,0.0645],[168,0.0722],[169,0.0836],[170,0.0857],[171,0.0852],[172,0.0857],[173,0.0856],[174,0.0902],[175,0.0924],[176,0.0931],[177,0.0931],[178,0.094],[179,0.0978],[180,0.0998],[181,0.1048],[182,0.1059],[183,0.1068],[184,0.1064],[185,0.108],[186,0.111],[187,0.1133],[188,0.1136],[189,0.1129],[190,0.1151],[191,0.1143],[192,0.1114],[193,0.1058],[194,0.1013],[195,0.0985],[196,0.1004],[197,0.1069],[198,0.1074],[199,0.1001],[200,0.0976],[201,0.0925],[202,0.0897],[203,0.0853],[204,0.0814],[205,0.0778],[206,0.0779],[207,0.0802],[208,0.0801],[209,0.0779],[211,0.076],[212,0.0769],[213,0.0782],[214,0.0783],[215,0.0798],[216,0.0816],[217,0.0843],[218,0.0867],[219,0.0889],[220,0.087],[221,0.0865],[222,0.0852],[223,0.0866],[224,0.086],[225,0.0879],[226,0.0875],[227,0.0901],[228,0.0902],[230,0.0958],[232,0.0982],[233,0.1002],[234,0.1008],[235,0.0962],[236,0.0926],[237,0.092],[238,0.0947],[240,0.0957],[241,0.1017],[242,0.1031],[243,0.1108],[244,0.1269],[245,0.1393],[246,0.1599],[247,0.1815],[248,0.193],[249,0.1944],[250,0.1887],[251,0.1846],[252,0.1778],[253,0.1807],[254,0.1759],[256,0.1632],[257,0.1636],[258,0.151],[259,0.1519],[260,0.1625],[261,0.1687],[262,0.1605],[263,0.1419],[264,0.1451],[265,0.1449],[266,0.1466],[267,0.1413],[268,0.1266],[269,0.1285],[270,0.1297],[271,0.1301],[272,0.1356],[273,0.1459],[274,0.1624],[275,0.1704],[276,0.1718],[277,0.1728],[278,0.1751],[279,0.1743],[280,0.1746],[281,0.1756],[282,0.1756],[283,0.1747],[284,0.1747],[285,0.1772],[286,0.1828],[287,0.1812],[288,0.1803],[290,0.1801],[291,0.1772],[292,0.1679],[293,0.1545],[295,0.1447],[296,0.1364],[297,0.1348],[298,0.1265],[299,0.1227],[300,0.112],[301,0.1115],[302,0.1083],[303,0.1058],[304,0.0985],[305,0.0949],[306,0.0925],[307,0.0913],[308,0.0907],[309,0.0919],[310,0.0924],[311,0.0937],[312,0.0937],[313,0.0922],[314,0.0927],[315,0.0885],[316,0.0885],[317,0.0882],[318,0.089],[319,0.089],[320,0.0899],[321,0.0913],[322,0.0911],[323,0.0915],[324,0.0922],[325,0.0915],[326,0.0911],[327,0.092],[328,0.0924],[329,0.0922],[330,0.0942],[331,0.0949],[332,0.0948],[333,0.0943],[334,0.0941],[335,0.0947],[337,0.0916],[338,0.0934],[339,0.0922],[340,0.0925],[341,0.0935],[342,0.0937],[343,0.0941],[345,0.0938],[347,0.097],[348,0.0958],[349,0.095],[350,0.0956],[351,0.0957],[352,0.0966],[353,0.0977],[355,0.0991],[356,0.0992],[357,0.0989],[358,0.1002],[359,0.0994],[360,0.0994]]
       }
       CTRL_R_jaw_ChinRaiseD {
        type=MovieSceneScriptingFloatChannel
        default=0
        thinned_from=492
        keys [[-131,0.0409],[-130,0.0412],[-129,0.0453],[-128,0.045],[-127,0.0467],[-126,0.0471],[-125,0.0508],[-124,0.0466],[-123,0.0417],[-122,0.0407],[-121,0.038],[-120,0.036],[-119,0.0356],[-118,0.0323],[-117,0.0324],[-116,0.033],[-115,0.029],[-114,0.0257],[-113,0.026],[-112,0.0245],[-111,0.0232],[-110,0.0235],[-109,0.0294],[-108,0.0297],[-107,0.0271],[-106,0.0269],[-105,0.0284],[-104,0.0295],[-103,0.0344],[-102,0.0323],[-101,0.0331],[-100,0.0347],[-99,0.0418],[-98,0.0428],[-97,0.041],[-96,0.0442],[-95,0.0452],[-94,0.0409],[-93,0.04],[-92,0.0367],[-91,0.0398],[-90,0.0475],[-89,0.0508],[-88,0.0434],[-87,0.0318],[-86,0.0257],[-85,0.0201],[-84,0.0161],[-83,0.0213],[-82,0.025],[-81,0.0192],[-80,0.0115],[-79,0.0116],[-78,0.013],[-76,0.0161],[-75,0.0383],[-74,0.0447],[-73,0.056],[-72,0.0648],[-71,0.0672],[-70,0.0711],[-69,0.0825],[-68,0.0748],[-67,0.0829],[-66,0.0839],[-65,0.0854],[-64,0.0835],[-63,0.0879],[-62,0.095],[-61,0.1017],[-60,0.1216],[-59,0.1239],[-58,0.124],[-57,0.1189],[-56,0.109],[-55,0.1048],[-54,0.0981],[-53,0.0953],[-52,0.0951],[-51,0.0902],[-50,0.0907],[-49,0.0908],[-48,0.0872],[-46,0.0877],[-45,0.0916],[-44,0.0926],[-43,0.097],[-42,0.0975],[-41,0.0912],[-40,0.0796],[-39,0.064],[-38,0.0604],[-37,0.0554],[-36,0.0615],[-35,0.068],[-34,0.0751],[-33,0.0895],[-32,0.1003],[-31,0.0984],[-30,0.0898],[-29,0.0871],[-28,0.0821],[-27,0.0831],[-26,0.0852],[-25,0.0883],[-24,0.0893],[-23,0.096],[-22,0.0994],[-21,0.0972],[-20,0.0912],[-19,0.0898],[-18,0.0908],[-17,0.0888],[-15,0.0905],[-14,0.0926],[-13,0.0929],[-12,0.0915],[-11,0.0888],[-10,0.0899],[-9,0.0915],[-8,0.0911],[-7,0.0903],[-6,0.091],[-5,0.0932],[-4,0.0942],[-3,0.0944],[-2,0.0941],[-1,0.0907],[0,0.0925],[1,0.0927],[2,0.0914],[3,0.0924],[4,0.0918],[6,0.0922],[7,0.0915],[8,0.0918],[9,0.0925],[10,0.0922],[11,0.0924],[12,0.091],[13,0.0893],[14,0.0896],[15,0.0891],[16,0.0894],[17,0.0888],[18,0.0884],[20,0.0872],[21,0.0875],[22,0.0893],[23,0.0896],[24,0.0903],[25,0.0906],[27,0.0882],[28,0.0882],[29,0.0892],[30,0.0894],[31,0.0882],[32,0.0877],[33,0.087],[35,0.0864],[36,0.085],[37,0.0866],[39,0.0851],[40,0.0862],[41,0.0853],[42,0.0849],[43,0.0859],[44,0.0863],[45,0.0862],[46,0.0879],[47,0.0876],[48,0.0863],[49,0.0861],[50,0.0846],[51,0.0816],[52,0.0783],[53,0.0821],[54,0.0829],[55,0.0833],[56,0.083],[57,0.0823],[58,0.0822],[59,0.0866],[60,0.0877],[61,0.0904],[62,0.0921],[63,0.0924],[64,0.0978],[65,0.0978],[67,0.0958],[68,0.0978],[69,0.0985],[70,0.0975],[71,0.097],[72,0.0979],[74,0.0969],[75,0.0967],[76,0.0952],[77,0.0974],[79,0.0974],[80,0.0948],[81,0.0955],[82,0.0936],[83,0.0934],[84,0.095],[85,0.0934],[86,0.0923],[87,0.0938],[88,0.0917],[89,0.0934],[90,0.0928],[91,0.0929],[93,0.0925],[94,0.0919],[95,0.0917],[96,0.0917],[97,0.0911],[98,0.0916],[99,0.0928],[101,0.0948],[102,0.0949],[103,0.0916],[104,0.0951],[105,0.0953],[106,0.0942],[107,0.0942],[108,0.0936],[109,0.0972],[110,0.0978],[111,0.0951],[112,0.0985],[113,0.103],[114,0.1019],[115,0.1012],[116,0.1026],[117,0.1007],[118,0.1034],[119,0.105],[120,0.1057],[121,0.104],[122,0.1046],[123,0.1061],[124,0.1047],[126,0.1015],[127,0.1039],[128,0.1037],[130,0.1077],[131,0.1067],[132,0.1046],[133,0.1038],[134,0.1089],[135,0.1178],[136,0.1271],[138,0.1402],[139,0.1332],[140,0.133],[141,0.1335],[142,0.1407],[143,0.1447],[144,0.1433],[145,0.143],[146,0.148],[147,0.1468],[148,0.1473],[149,0.1449],[150,0.145],[151,0.1446],[152,0.1407],[153,0.1444],[154,0.144],[155,0.144],[156,0.1425],[157,0.138],[158,0.1365],[159,0.1383],[160,0.1369],[161,0.1363],[162,0.1336],[163,0.1373],[164,0.1343],[165,0.1339],[166,0.1321],[169,0.1224],[170,0.1171],[171,0.1152],[172,0.1127],[173,0.1087],[175,0.1149],[176,0.113],[177,0.1133],[178,0.1142],[179,0.1143],[181,0.1177],[183,0.1162],[184,0.1138],[185,0.1123],[186,0.1124],[187,0.1131],[188,0.1131],[189,0.1098],[190,0.1101],[191,0.1088],[192,0.109],[193,0.1061],[194,0.1085],[195,0.1056],[196,0.0995],[198,0.1033],[199,0.105],[200,0.1059],[201,0.1186],[202,0.1112],[203,0.1195],[204,0.124],[205,0.1245],[206,0.1216],[207,0.1214],[208,0.1199],[209,0.1188],[210,0.1203],[211,0.1188],[212,0.1193],[213,0.1192],[214,0.1172],[215,0.1161],[216,0.1158],[217,0.1141],[218,0.1153],[219,0.1159],[220,0.1149],[221,0.1153],[222,0.1128],[223,0.1134],[224,0.1136],[226,0.1123],[227,0.1121],[228,0.1098],[229,0.1099],[230,0.1136],[231,0.1145],[232,0.1143],[233,0.1189],[234,0.1205],[235,0.1144],[236,0.0958],[237,0.086],[238,0.0824],[239,0.0814],[240,0.0775],[241,0.0731],[242,0.0725],[243,0.073],[244,0.0741],[245,0.0732],[246,0.0577],[247,0.0336],[248,0.0145],[249,0.007],[250,0.0055],[251,0.0077],[252,0.0279],[253,0.0377],[254,0.0444],[255,0.043],[256,0.048],[257,0.0368],[258,0.0246],[259,0.0244],[260,0.0238],[261,0.0288],[262,0.035],[263,0.0589],[265,0.0645],[266,0.0627],[267,0.0614],[268,0.0635],[269,0.0641],[270,0.0641],[271,0.0654],[272,0.0664],[273,0.0595],[274,0.0537],[275,0.0513],[276,0.0501],[277,0.0482],[278,0.0468],[279,0.0457],[280,0.0461],[281,0.0472],[282,0.045],[283,0.0481],[284,0.0484],[285,0.0443],[286,0.0441],[287,0.0473],[288,0.0511],[289,0.0515],[290,0.0499],[291,0.0494],[292,0.0459],[293,0.0398],[294,0.0383],[295,0.0458],[296,0.0523],[297,0.0537],[298,0.0675],[299,0.0785],[300,0.093],[301,0.0937],[302,0.0974],[303,0.0976],[304,0.0975],[305,0.0957],[306,0.0951],[307,0.0953],[308,0.0948],[310,0.0957],[311,0.0979],[313,0.0981],[314,0.0959],[315,0.0968],[317,0.098],[318,0.0987],[319,0.0992],[321,0.0993],[322,0.0983],[323,0.098],[324,0.0994],[326,0.0998],[327,0.0992],[328,0.1005],[329,0.1008],[330,0.1025],[331,0.1032],[332,0.1019],[333,0.1016],[334,0.1008],[335,0.102],[336,0.1013],[337,0.1],[338,0.0988],[339,0.1002],[340,0.1002],[341,0.1007],[342,0.1007],[343,0.0989],[344,0.0992],[345,0.0985],[347,0.1013],[348,0.1],[349,0.1006],[350,0.1006],[351,0.1018],[352,0.1012],[353,0.1027],[354,0.102],[355,0.101],[356,0.1008],[357,0.1002],[358,0.1012],[359,0.1039],[360,0.1039]]
       }
       CTRL_R_jaw_ChinRaiseU {type=MovieSceneScriptingFloatChannel, default=0, const [0,492]}
       CTRL_L_jaw_ChinRaiseU {type=MovieSceneScriptingFloatChannel, default=0, const [0,492]}
       CTRL_L_jaw_ChinRaiseD {
        type=MovieSceneScriptingFloatChannel
        default=0
        thinned_from=492
        keys [[-131,0.0389],[-130,0.0388],[-129,0.043],[-128,0.0426],[-127,0.0442],[-126,0.0444],[-125,0.0479],[-124,0.0439],[-123,0.0391],[-122,0.0382],[-121,0.0355],[-120,0.0336],[-119,0.033],[-118,0.0298],[-117,0.0299],[-116,0.0305],[-115,0.0268],[-114,0.0236],[-113,0.0239],[-112,0.0225],[-111,0.0212],[-110,0.0213],[-109,0.0269],[-108,0.0274],[-107,0.0249],[-106,0.0247],[-105,0.0262],[-104,0.0274],[-103,0.032],[-102,0.03],[-101,0.0307],[-100,0.0323],[-99,0.039],[-98,0.04],[-97,0.0383],[-96,0.0413],[-95,0.0422],[-94,0.0382],[-93,0.0373],[-92,0.0344],[-91,0.0372],[-90,0.0446],[-89,0.0481],[-88,0.0409],[-87,0.0298],[-86,0.0241],[-85,0.0189],[-84,0.0151],[-83,0.0203],[-82,0.0239],[-81,0.0187],[-80,0.0118],[-79,0.012],[-78,0.0139],[-77,0.0163],[-76,0.018],[-75,0.0394],[-74,0.0456],[-73,0.0566],[-72,0.0649],[-71,0.0673],[-70,0.0713],[-69,0.082],[-68,0.0749],[-67,0.0826],[-66,0.0834],[-65,0.0846],[-64,0.0829],[-63,0.087],[-61,0.1009],[-60,0.1207],[-59,0.1239],[-58,0.1254],[-57,0.1205],[-56,0.1111],[-55,0.1073],[-54,0.1009],[-53,0.0983],[-52,0.0982],[-51,0.0933],[-49,0.0936],[-48,0.0899],[-47,0.0896],[-46,0.0896],[-45,0.093],[-44,0.0939],[-43,0.0976],[-42,0.0977],[-41,0.0916],[-40,0.0801],[-39,0.0646],[-38,0.061],[-37,0.0557],[-36,0.0616],[-35,0.0679],[-34,0.0745],[-33,0.0888],[-32,0.1002],[-31,0.0987],[-30,0.0918],[-29,0.0893],[-28,0.0847],[-27,0.0859],[-26,0.0882],[-25,0.0916],[-24,0.0928],[-23,0.0994],[-22,0.103],[-21,0.1005],[-20,0.0944],[-19,0.0925],[-18,0.0933],[-17,0.0911],[-15,0.0926],[-14,0.0945],[-13,0.0949],[-12,0.0936],[-11,0.091],[-10,0.0919],[-9,0.0938],[-8,0.0935],[-7,0.0926],[-6,0.0935],[-5,0.0959],[-4,0.0968],[-2,0.0971],[-1,0.0936],[0,0.0951],[1,0.0953],[2,0.0942],[3,0.0951],[4,0.0947],[5,0.0945],[6,0.0946],[7,0.0939],[8,0.0942],[9,0.0949],[10,0.0946],[11,0.0946],[13,0.0916],[14,0.0918],[15,0.0914],[16,0.092],[17,0.0913],[18,0.091],[19,0.0903],[20,0.0898],[21,0.09],[22,0.0916],[23,0.0919],[24,0.0927],[25,0.0932],[27,0.0909],[28,0.0908],[29,0.0918],[30,0.0919],[31,0.0908],[32,0.0903],[33,0.0893],[34,0.0888],[35,0.0887],[36,0.0878],[37,0.0892],[39,0.0875],[40,0.0886],[41,0.0879],[42,0.0876],[43,0.0886],[44,0.0891],[45,0.0892],[46,0.0907],[47,0.0905],[48,0.0893],[49,0.0887],[50,0.0872],[52,0.0805],[53,0.0844],[54,0.0853],[55,0.0857],[56,0.0852],[57,0.0844],[58,0.084],[59,0.0883],[60,0.0894],[61,0.0923],[62,0.0939],[63,0.0942],[64,0.0995],[65,0.0995],[66,0.0987],[67,0.0976],[68,0.0997],[69,0.1005],[70,0.0994],[71,0.0992],[72,0.1001],[73,0.0994],[74,0.0988],[75,0.0988],[76,0.0974],[77,0.0998],[79,0.1],[80,0.0976],[81,0.0983],[82,0.0968],[83,0.0969],[84,0.0981],[85,0.0966],[86,0.0957],[87,0.0968],[88,0.0948],[89,0.0962],[90,0.0955],[92,0.0957],[93,0.0955],[94,0.0949],[95,0.0946],[96,0.0946],[97,0.094],[98,0.0942],[99,0.0956],[101,0.0975],[102,0.0976],[103,0.0947],[104,0.0981],[105,0.0983],[106,0.0969],[107,0.0969],[108,0.0963],[109,0.0999],[110,0.1003],[111,0.0974],[112,0.1009],[113,0.1051],[115,0.1035],[116,0.1046],[117,0.1029],[118,0.1054],[119,0.107],[120,0.1079],[121,0.1064],[122,0.1069],[123,0.1083],[124,0.1071],[126,0.1041],[127,0.1064],[128,0.1063],[130,0.1106],[131,0.1095],[132,0.1073],[133,0.1069],[134,0.112],[136,0.1302],[138,0.1443],[139,0.1369],[141,0.1375],[142,0.1446],[143,0.1484],[144,0.147],[145,0.1464],[146,0.1515],[147,0.1504],[148,0.1508],[149,0.1484],[151,0.1482],[152,0.1443],[153,0.148],[155,0.1475],[156,0.1461],[157,0.1415],[158,0.1401],[159,0.142],[160,0.1404],[161,0.14],[162,0.1371],[163,0.141],[164,0.1382],[165,0.1377],[166,0.136],[167,0.1325],[168,0.1288],[169,0.1253],[170,0.1199],[171,0.1176],[172,0.115],[173,0.1107],[174,0.1134],[175,0.1164],[176,0.1148],[177,0.1151],[178,0.1158],[179,0.1155],[181,0.1187],[182,0.1182],[183,0.1174],[184,0.1152],[185,0.1136],[186,0.1137],[187,0.1142],[188,0.1142],[189,0.1111],[190,0.1113],[191,0.1101],[192,0.11],[193,0.1073],[194,0.1098],[195,0.1071],[196,0.1006],[197,0.1019],[199,0.1056],[200,0.1065],[201,0.119],[202,0.1119],[203,0.1197],[204,0.1242],[205,0.1243],[206,0.1217],[207,0.1215],[208,0.1198],[209,0.1191],[210,0.1209],[211,0.1197],[212,0.12],[213,0.12],[214,0.1178],[215,0.1168],[216,0.1166],[217,0.1152],[218,0.1165],[219,0.1168],[220,0.116],[221,0.1165],[222,0.114],[223,0.1147],[224,0.1148],[225,0.114],[227,0.1134],[228,0.1108],[229,0.1109],[230,0.1142],[231,0.1153],[232,0.1151],[233,0.1193],[234,0.1205],[235,0.1138],[236,0.0955],[237,0.086],[238,0.0826],[239,0.0813],[240,0.0774],[241,0.073],[242,0.0724],[243,0.0727],[244,0.0738],[245,0.0726],[246,0.0569],[247,0.0329],[248,0.0137],[249,0.0064],[250,0.0049],[251,0.0072],[252,0.0263],[253,0.0357],[254,0.0423],[255,0.041],[256,0.0456],[257,0.0348],[258,0.0227],[260,0.0223],[261,0.0274],[262,0.033],[263,0.0558],[264,0.0585],[265,0.0616],[266,0.0599],[267,0.0586],[268,0.0606],[269,0.0614],[270,0.0613],[272,0.0638],[273,0.0572],[274,0.0513],[275,0.0489],[276,0.0477],[277,0.0458],[279,0.0431],[280,0.0437],[281,0.0446],[282,0.0424],[283,0.0453],[284,0.0456],[285,0.0418],[286,0.0413],[287,0.0445],[288,0.0483],[289,0.0487],[290,0.0471],[291,0.0466],[292,0.0433],[293,0.0371],[294,0.0356],[295,0.0431],[296,0.0498],[297,0.0512],[298,0.0655],[299,0.077],[300,0.0921],[301,0.0929],[302,0.0975],[304,0.0982],[305,0.0967],[306,0.0964],[307,0.0968],[308,0.0962],[310,0.0971],[311,0.099],[313,0.0995],[314,0.0972],[315,0.0981],[317,0.0994],[318,0.1004],[319,0.1008],[321,0.1006],[322,0.0995],[323,0.0994],[324,0.1009],[325,0.1013],[327,0.1008],[328,0.1022],[329,0.1026],[330,0.1039],[331,0.1046],[332,0.1033],[333,0.1029],[334,0.1022],[335,0.1035],[336,0.1028],[337,0.1014],[338,0.1003],[339,0.1015],[340,0.1013],[341,0.102],[342,0.102],[343,0.1002],[344,0.1005],[345,0.0999],[347,0.1027],[348,0.1015],[349,0.1019],[350,0.1017],[351,0.103],[352,0.1024],[353,0.1037],[354,0.1033],[355,0.1024],[356,0.1021],[357,0.1014],[358,0.1026],[359,0.1054],[360,0.1054]]
       }
       CTRL_R_mouth_lipsBlow {type=MovieSceneScriptingFloatChannel, default=0, const [0,492]}
       CTRL_L_mouth_lipsTogetherU {
        type=MovieSceneScriptingFloatChannel
        default=0
        thinned_from=492
        keys [[-131,0.4635],[-130,0.4034],[-129,0.3622],[-127,0.3469],[-126,0.3536],[-125,0.3537],[-124,0.3378],[-122,0.3408],[-121,0.3405],[-120,0.3493],[-119,0.3358],[-118,0.3403],[-117,0.336],[-116,0.3048],[-115,0.2962],[-114,0.2987],[-113,0.2916],[-112,0.2801],[-111,0.272],[-110,0.2801],[-109,0.2568],[-108,0.2543],[-107,0.2478],[-106,0.2462],[-105,0.2428],[-104,0.2534],[-103,0.2521],[-102,0.2527],[-101,0.2715],[-100,0.2701],[-99,0.2745],[-98,0.2714],[-97,0.2692],[-96,0.2729],[-95,0.2757],[-94,0.3162],[-93,0.3181],[-92,0.3232],[-91,0.3404],[-90,0.3755],[-89,0.3939],[-88,0.4477],[-87,0.5355],[-85,0.6326],[-84,0.6847],[-83,0.7058],[-82,0.7621],[-81,0.8139],[-80,0.8619],[-79,0.8638],[-78,0.8816],[-77,0.9266],[-76,0.9491],[-75,0.9051],[-74,0.8761],[-73,0.8696],[-72,0.8455],[-71,0.8349],[-70,0.8261],[-69,0.8244],[-68,0.8393],[-67,0.8081],[-66,0.8294],[-65,0.8052],[-64,0.8217],[-63,0.8165],[-62,0.7991],[-61,0.7595],[-60,0.6138],[-59,0.3091],[-58,0.0375],[-57,0.0395],[-56,0.0424],[-55,0.0438],[-54,0.0423],[-53,0.0449],[-52,0.0488],[-51,0.0433],[-50,0.0446],[-49,0.045],[-48,0.0423],[-47,0.0382],[-46,0.0461],[-45,0.1926],[-44,0.2566],[-43,0.4842],[-42,0.6556],[-41,0.7204],[-39,0.804],[-38,0.8159],[-37,0.8779],[-36,0.8551],[-35,0.8155],[-34,0.834],[-33,0.7355],[-32,0.439],[-31,0.3461],[-30,0.0419],[-29,0.0357],[-28,0.0361],[-27,0.0419],[-26,0.0383],[-25,0.0372],[-24,0.0363],[-23,0.0294],[-22,0.0363],[-21,0.0838],[-20,0.0351],[-19,0.0313],[-18,0.1055],[-17,0.0523],[-16,0.0487],[-15,0.0515],[-14,0.1018],[-13,0.0588],[-12,0.0347],[-11,0.0363],[-10,0.0396],[-9,0.0391],[-8,0.0456],[-7,0.0398],[-6,0.0409],[-5,0.04],[-4,0.0365],[-3,0.0409],[-2,0.0392],[-1,0.0394],[0,0.0432],[1,0.0415],[2,0.045],[3,0.0358],[4,0.0489],[5,0.043],[6,0.0426],[7,0.0479],[8,0.0465],[9,0.0444],[10,0.049],[11,0.0472],[12,0.0465],[13,0.0492],[14,0.0452],[15,0.049],[16,0.0475],[17,0.0462],[18,0.0483],[19,0.0428],[20,0.0425],[21,0.0471],[22,0.0388],[23,0.0395],[24,0.0422],[25,0.0422],[26,0.0432],[27,0.0436],[28,0.0432],[29,0.0454],[30,0.0444],[32,0.0434],[33,0.0401],[34,0.049],[35,0.0479],[36,0.0414],[37,0.0468],[38,0.0453],[40,0.0476],[41,0.0452],[42,0.0486],[43,0.0479],[44,0.0456],[45,0.0459],[46,0.0526],[47,0.0417],[48,0.0401],[49,0.0416],[50,0.0354],[51,0.0463],[52,0.0397],[53,0.0388],[54,0.0402],[55,0.0539],[56,0.0803],[57,0.0361],[58,0.042],[59,0.1224],[60,0.1392],[61,0.0985],[62,0.125],[63,0.1215],[64,0.0654],[65,0.0654],[66,0.0728],[67,0.0532],[68,0.1199],[69,0.0575],[70,0.0353],[71,0.0212],[72,0.0167],[73,0.0111],[74,0.0163],[75,0.0285],[76,0.0197],[77,0.024],[78,0.0234],[79,0.0142],[80,0.015],[81,0.019],[82,0.0056],[83,0.0101],[84,0.0051],[85,0.0153],[86,0.0159],[87,0.0145],[88,0.0002],[89,0.0042],[90,0.0052],[92,0.0052],[93,0.0037],[94,0.0014],[95,0.0034],[96,0],[97,0],[98,0.0026],[99,0.0128],[100,0.0143],[101,0.0066],[102,0.0047],[103,0],[105,0],[106,0.0013],[107,0.0049],[108,0.0003],[109,0.0047],[110,0.0046],[111,0.0096],[112,0.0093],[113,0.0083],[114,0.0003],[115,0.0179],[117,0.0003],[118,0.0082],[119,0.0013],[120,0.0079],[121,0.0132],[123,0.0042],[124,0],[126,0],[127,0.0045],[128,0],[133,0],[134,0.015],[135,0.015],[136,0.0183],[137,0.033],[138,0.0418],[139,0.0002],[140,0.0318],[141,0.0196],[142,0.0258],[143,0.031],[145,0.0176],[146,0.0093],[147,0.0216],[148,0.0135],[149,0.0093],[150,0.0136],[151,0.0234],[152,0.02],[153,0.0237],[154,0.0189],[155,0.0222],[156,0.0097],[157,0.0154],[158,0.0156],[159,0.029],[160,0.0109],[161,0.0244],[162,0.0261],[163,0.0236],[164,0.0125],[165,0.0203],[166,0.0252],[167,0.0294],[168,0.0142],[169,0.0305],[170,0.0317],[171,0.0378],[172,0.0339],[173,0.0511],[174,0.1907],[176,0.2942],[177,0.2835],[178,0.2535],[179,0.2875],[180,0.3092],[181,0.3235],[182,0.324],[183,0.3139],[184,0.2779],[185,0.2875],[186,0.308],[188,0.3089],[189,0.3309],[190,0.3505],[191,0.3799],[192,0.4289],[193,0.3458],[194,0.4363],[195,0.4035],[196,0.52],[197,0.6683],[198,0.733],[199,0.7625],[200,0.7655],[201,0.6828],[202,0.7112],[203,0.6772],[204,0.6735],[205,0.6994],[206,0.683],[207,0.6842],[208,0.6509],[209,0.6257],[210,0.5368],[211,0.5134],[212,0.5348],[213,0.5369],[214,0.5001],[215,0.5176],[216,0.5194],[217,0.4917],[218,0.4995],[219,0.4785],[220,0.485],[221,0.4725],[222,0.4338],[223,0.4752],[224,0.4563],[225,0.4825],[226,0.4753],[227,0.4783],[228,0.4938],[229,0.5124],[230,0.5416],[231,0.5508],[232,0.5512],[233,0.6332],[234,0.656],[235,0.6326],[236,0.6427],[237,0.6147],[238,0.6044],[239,0.6517],[240,0.5921],[241,0.6739],[242,0.6977],[243,0.7173],[244,0.7101],[245,0.7175],[246,0.7896],[247,0.781],[248,0.6886],[249,0.6506],[250,0.5907],[251,0.5722],[252,0.5133],[253,0.505],[254,0.5106],[255,0.5287],[256,0.4754],[257,0.5068],[258,0.5819],[259,0.5736],[260,0.5205],[261,0.5325],[262,0.4385],[263,0.2669],[264,0.2909],[265,0.3191],[266,0.3111],[267,0.2715],[268,0.2589],[269,0.2886],[270,0.2944],[271,0.3202],[272,0.3581],[273,0.3948],[274,0.3955],[275,0.3984],[277,0.3775],[278,0.3696],[279,0.3694],[280,0.3536],[281,0.353],[282,0.371],[283,0.3598],[284,0.3579],[285,0.3621],[286,0.3714],[288,0.3777],[289,0.3865],[290,0.3899],[291,0.3635],[292,0.3335],[293,0.3115],[294,0.2864],[295,0.3462],[296,0.4492],[297,0.4703],[298,0.6158],[299,0.6485],[300,0.6024],[301,0.5811],[302,0.415],[303,0.3763],[304,0.2528],[305,0.196],[306,0.1661],[307,0.1586],[308,0.129],[309,0.1344],[310,0.146],[311,0.1349],[312,0.115],[313,0.0839],[314,0.0439],[315,0.0472],[316,0.0442],[317,0.0273],[318,0.0427],[319,0.0394],[320,0.0802],[321,0.0744],[322,0.0611],[323,0.0794],[324,0.0737],[325,0.0433],[326,0.0966],[327,0.087],[328,0.0874],[329,0.1099],[330,0.1699],[331,0.1486],[332,0.087],[333,0.1202],[334,0.0948],[335,0.1253],[336,0.0757],[337,0.0495],[338,0.0895],[339,0.0997],[340,0.1175],[341,0.099],[342,0.1179],[343,0.1143],[344,0.0815],[345,0.1194],[346,0.1313],[347,0.1533],[348,0.144],[349,0.117],[350,0.1313],[351,0.1179],[352,0.1075],[353,0.1386],[354,0.1213],[355,0.1132],[356,0.1127],[357,0.1078],[358,0.0741],[359,0.0946],[360,0.0946]]
       }
       CTRL_L_mouth_lipsTogetherD {
        type=MovieSceneScriptingFloatChannel
        default=0
        thinned_from=492
        keys [[-131,0.1906],[-130,0.2114],[-129,0.2128],[-128,0.2076],[-127,0.197],[-126,0.2008],[-125,0.195],[-124,0.1863],[-123,0.1834],[-122,0.183],[-121,0.1785],[-120,0.1797],[-119,0.1784],[-118,0.1829],[-117,0.1841],[-116,0.1887],[-115,0.1775],[-114,0.1646],[-113,0.1511],[-112,0.1357],[-111,0.1242],[-110,0.1336],[-109,0.1518],[-108,0.1567],[-107,0.1379],[-106,0.1355],[-105,0.1467],[-104,0.1644],[-103,0.1799],[-102,0.1654],[-101,0.1532],[-100,0.1687],[-99,0.1806],[-98,0.1708],[-97,0.1653],[-96,0.1739],[-95,0.1768],[-94,0.1727],[-93,0.1731],[-92,0.1859],[-91,0.202],[-89,0.2603],[-88,0.2771],[-87,0.2868],[-86,0.3015],[-85,0.3113],[-84,0.3246],[-83,0.3364],[-82,0.334],[-81,0.3094],[-80,0.2274],[-79,0.2177],[-78,0.1322],[-77,0.0383],[-76,0.0095],[-75,0.0079],[-74,0.0087],[-73,0.011],[-71,0.011],[-70,0.0108],[-69,0.0111],[-68,0.0111],[-67,0.0123],[-65,0.0116],[-64,0.0126],[-63,0.0119],[-62,0.0138],[-61,0.0153],[-60,0.0217],[-59,0.0341],[-58,0.0458],[-57,0.0483],[-56,0.0522],[-55,0.0491],[-54,0.0487],[-53,0.0493],[-51,0.0484],[-50,0.0541],[-49,0.055],[-48,0.0515],[-47,0.0425],[-46,0.0406],[-45,0.0313],[-44,0.0422],[-43,0.0792],[-41,0.1539],[-40,0.1742],[-39,0.136],[-38,0.1151],[-37,0.0107],[-36,0.0095],[-35,0.0108],[-34,0.0128],[-33,0.0174],[-32,0.0326],[-31,0.0347],[-30,0.0438],[-29,0.0402],[-28,0.0408],[-27,0.0467],[-26,0.0401],[-25,0.041],[-24,0.0415],[-23,0.0336],[-22,0.0389],[-21,0.0305],[-20,0.0387],[-19,0.0382],[-18,0.0381],[-17,0.0404],[-16,0.041],[-15,0.0404],[-14,0.035],[-13,0.0404],[-11,0.0443],[-10,0.0411],[-8,0.054],[-7,0.0486],[-6,0.0499],[-5,0.0489],[-4,0.0411],[-3,0.0454],[-2,0.0441],[-1,0.049],[0,0.0481],[1,0.0506],[2,0.055],[3,0.0549],[4,0.0539],[5,0.052],[6,0.0472],[7,0.048],[8,0.0479],[9,0.0491],[10,0.0539],[11,0.0521],[12,0.0515],[13,0.0541],[14,0.0524],[15,0.0525],[16,0.0476],[17,0.0499],[18,0.0538],[19,0.0564],[20,0.0476],[21,0.0518],[22,0.0436],[23,0.0431],[24,0.0435],[26,0.0501],[27,0.051],[28,0.0481],[30,0.0526],[31,0.0501],[32,0.0523],[33,0.0407],[34,0.049],[35,0.0491],[36,0.0498],[37,0.0472],[38,0.0471],[39,0.0515],[40,0.0538],[41,0.0514],[42,0.0508],[43,0.0529],[44,0.0507],[45,0.051],[46,0.0579],[47,0.0463],[48,0.04],[49,0.0416],[50,0.0392],[51,0.0371],[52,0.0436],[53,0.0303],[54,0.0378],[55,0.0368],[56,0.0312],[57,0.0158],[58,0.0027],[59,0.0143],[60,0.0178],[61,0.0151],[62,0.0184],[63,0.0187],[64,0.0231],[65,0.0231],[66,0.0268],[67,0.0204],[68,0.0149],[69,0.0235],[70,0.016],[71,0.0207],[72,0.0169],[73,0.0154],[74,0.0206],[75,0.0292],[76,0.0196],[77,0.024],[78,0.0234],[79,0.0142],[80,0.015],[81,0.0191],[82,0.0105],[83,0.0153],[84,0.0147],[85,0.0204],[86,0.0213],[87,0.0195],[88,0.0054],[89,0.0088],[90,0],[91,0.0003],[92,0.0052],[93,0.0053],[94,0.0067],[95,0.0102],[96,0.0083],[97,0.0041],[98,0.0026],[99,0.0153],[100,0.0193],[101,0.0115],[102,0.0096],[103,0.0047],[104,0.0049],[105,0.0049],[106,0.0062],[107,0.0098],[108,0.0003],[109,0.0047],[110,0.0089],[111,0.0096],[112,0.0093],[113,0.0041],[114,0.0001],[115,0.018],[116,0.0135],[117,0.0047],[118,0.0083],[119,0.0049],[120,0.0081],[121,0.0131],[122,0.0088],[123,0.0047],[124,0.0089],[125,0.0004],[126,0],[127,0.0045],[128,0],[131,0],[132,0.0002],[133,0.0054],[134,0.0248],[135,0.02],[136,0.0186],[137,0.0378],[138,0.0523],[139,0.0003],[140,0.0371],[141,0.0252],[142,0.031],[143,0.0323],[144,0.0196],[145,0.0176],[146,0.0097],[147,0.0301],[148,0.0136],[149,0.0138],[150,0.0182],[151,0.0237],[152,0.025],[153,0.0285],[154,0.0236],[155,0.0266],[156,0.0098],[157,0.0204],[158,0.0207],[159,0.0339],[160,0.0156],[161,0.0255],[162,0.0352],[163,0.0238],[164,0.0175],[165,0.0255],[166,0.0202],[167,0.0245],[168,0.0184],[169,0.0323],[170,0.034],[171,0.0363],[172,0.0305],[173,0.0136],[174,0.0091],[175,0.0144],[176,0.017],[177,0.0185],[178,0.021],[179,0.0197],[180,0.0295],[181,0.0649],[182,0.0667],[183,0.0672],[184,0.0421],[185,0.0628],[186,0.1002],[187,0.1062],[188,0.1065],[189,0.1112],[190,0.1119],[191,0.0875],[192,0.0582],[193,0.0347],[194,0.036],[195,0.0333],[196,0.0717],[197,0.1205],[198,0.1102],[199,0.1095],[200,0.1145],[201,0.1934],[202,0.1283],[203,0.121],[204,0.113],[205,0.0952],[206,0.096],[207,0.1238],[208,0.0945],[209,0.0716],[210,0.0564],[211,0.0299],[212,0.0381],[213,0.0305],[214,0.0176],[215,0.0273],[216,0.0357],[217,0.028],[218,0.0371],[219,0.0347],[220,0.0388],[221,0.0256],[222,0.0218],[223,0.0453],[224,0.0323],[225,0.0275],[226,0.0329],[227,0.0409],[228,0.0219],[229,0.052],[230,0.0724],[231,0.0999],[232,0.1221],[233,0.1841],[234,0.2028],[235,0.175],[236,0.0425],[237,0.0243],[238,0.0483],[239,0.059],[240,0.042],[241,0.0848],[242,0.0972],[243,0.1256],[244,0.176],[245,0.2338],[246,0.2569],[247,0.236],[248,0.1598],[249,0.1245],[250,0.1432],[251,0.1641],[252,0.2417],[253,0.2927],[254,0.2816],[255,0.258],[256,0.2717],[257,0.2444],[258,0.131],[259,0.1382],[260,0.2018],[261,0.2049],[262,0.1779],[263,0.2124],[264,0.2297],[265,0.2229],[266,0.2255],[267,0.1955],[268,0.1498],[269,0.2008],[270,0.2133],[271,0.2203],[272,0.2234],[273,0.2306],[274,0.2494],[275,0.2598],[276,0.248],[277,0.2455],[278,0.2421],[279,0.2274],[280,0.2339],[281,0.2368],[282,0.219],[283,0.2336],[284,0.2359],[285,0.2274],[286,0.2539],[287,0.2493],[288,0.2471],[289,0.2417],[290,0.239],[291,0.2352],[292,0.2092],[293,0.1824],[294,0.1809],[295,0.2184],[296,0.2349],[297,0.2355],[298,0.2102],[299,0.1936],[300,0.1293],[301,0.1191],[302,0.0357],[303,0.033],[304,0.0255],[305,0.0248],[307,0.0322],[308,0.026],[309,0.0244],[310,0.0249],[311,0.021],[312,0.021],[313,0.0182],[314,0.0201],[315,0.0238],[316,0.0241],[317,0.0156],[318,0.0237],[319,0.0273],[320,0.0262],[321,0.0222],[322,0.0118],[323,0.0154],[324,0.0254],[325,0.0227],[326,0.0218],[327,0.0184],[328,0.0251],[329,0.018],[330,0.0205],[331,0.0237],[332,0.0254],[333,0.0183],[334,0.0222],[335,0.0239],[336,0.0218],[337,0.026],[338,0.0284],[339,0.0216],[340,0.0207],[341,0.0177],[342,0.021],[343,0.0249],[344,0.0182],[345,0.0211],[346,0.0171],[347,0.0167],[348,0.0173],[349,0.0168],[350,0.014],[351,0.0243],[352,0.0215],[353,0.0236],[354,0.0221],[355,0.0236],[356,0.0246],[357,0.0212],[358,0.0221],[359,0.0176],[360,0.0176]]
       }
       CTRL_R_mouth_lipsTogetherU {
        type=MovieSceneScriptingFloatChannel
        default=0
        thinned_from=492
        keys [[-131,0.4669],[-130,0.4069],[-129,0.3658],[-127,0.3495],[-126,0.3567],[-125,0.3566],[-124,0.3404],[-122,0.343],[-121,0.3429],[-120,0.3517],[-119,0.3389],[-118,0.3417],[-117,0.3376],[-116,0.3075],[-115,0.2994],[-114,0.3014],[-113,0.2938],[-112,0.2823],[-111,0.2738],[-110,0.2815],[-109,0.2583],[-108,0.2552],[-107,0.2492],[-106,0.2477],[-105,0.2444],[-104,0.2549],[-103,0.2538],[-102,0.2555],[-101,0.2737],[-100,0.2709],[-99,0.276],[-97,0.2694],[-96,0.274],[-95,0.2772],[-94,0.3177],[-93,0.3195],[-92,0.3249],[-91,0.3423],[-90,0.3784],[-89,0.3954],[-88,0.45],[-87,0.5387],[-85,0.6375],[-84,0.6897],[-83,0.7099],[-82,0.7669],[-81,0.8192],[-80,0.8687],[-79,0.8706],[-78,0.8884],[-77,0.9324],[-76,0.9545],[-75,0.9115],[-74,0.881],[-73,0.874],[-72,0.8486],[-71,0.8402],[-70,0.8291],[-69,0.8279],[-68,0.8422],[-67,0.8111],[-66,0.8328],[-65,0.8086],[-64,0.8246],[-63,0.8189],[-62,0.8012],[-61,0.761],[-60,0.6155],[-59,0.312],[-58,0.0375],[-57,0.0378],[-56,0.0375],[-55,0.0387],[-53,0.0416],[-52,0.0434],[-51,0.0381],[-50,0.0396],[-49,0.04],[-48,0.0376],[-47,0.034],[-46,0.0415],[-45,0.1957],[-44,0.2605],[-43,0.4897],[-42,0.6598],[-41,0.7261],[-40,0.767],[-39,0.8094],[-38,0.8213],[-37,0.8824],[-36,0.8591],[-35,0.8198],[-34,0.838],[-33,0.7397],[-32,0.4391],[-31,0.3459],[-30,0.0375],[-29,0.0328],[-28,0.0341],[-27,0.0398],[-26,0.0383],[-24,0.0363],[-23,0.0254],[-22,0.0363],[-21,0.084],[-20,0.0387],[-19,0.0382],[-18,0.1086],[-17,0.0524],[-16,0.0525],[-15,0.0588],[-14,0.1018],[-13,0.0588],[-12,0.0386],[-11,0.0403],[-10,0.0403],[-9,0.0386],[-8,0.0419],[-7,0.0398],[-6,0.0409],[-5,0.0401],[-4,0.0365],[-3,0.0409],[-2,0.0392],[-1,0.0392],[0,0.0385],[1,0.0405],[2,0.045],[3,0.04],[4,0.0392],[5,0.0381],[6,0.0471],[7,0.0434],[8,0.043],[9,0.0416],[10,0.0427],[11,0.0452],[12,0.0397],[13,0.0432],[14,0.0424],[15,0.0412],[16,0.0416],[17,0.0415],[18,0.0438],[19,0.0425],[20,0.0425],[21,0.0471],[22,0.0388],[23,0.0395],[24,0.0422],[25,0.0421],[26,0.0432],[27,0.0436],[28,0.0432],[29,0.0425],[31,0.0437],[32,0.0393],[33,0.0395],[34,0.0441],[35,0.0436],[36,0.0399],[37,0.0396],[38,0.0403],[40,0.0407],[42,0.0456],[43,0.0478],[44,0.0456],[45,0.0459],[46,0.0425],[47,0.0417],[48,0.0389],[49,0.0374],[50,0.0315],[51,0.0463],[52,0.0397],[53,0.0388],[54,0.0401],[55,0.0584],[56,0.0836],[57,0.0389],[58,0.0433],[59,0.1224],[60,0.1392],[61,0.0984],[62,0.125],[63,0.1212],[64,0.0615],[65,0.0615],[66,0.0727],[67,0.0493],[68,0.1163],[69,0.0574],[70,0.0314],[71,0.0211],[72,0.0169],[73,0.0143],[74,0.0173],[75,0.0285],[76,0.0153],[77,0.0192],[78,0.0189],[79,0.0141],[80,0.0102],[81,0.0143],[82,0.0054],[83,0.01],[84,0.0003],[85,0.0053],[86,0.0105],[87,0.0144],[88,0.0002],[89,0.0039],[90,0],[93,0],[94,0.0014],[95,0.0034],[96,0],[97,0],[98,0.0026],[99,0.0128],[100,0.0171],[101,0.0081],[102,0.0047],[103,0],[104,0.0046],[105,0.0049],[106,0.0026],[107,0.0098],[108,0.0003],[109,0.0047],[110,0.0002],[111,0.0048],[112,0.0046],[113,0],[114,0],[115,0.0134],[116,0.0047],[117,0.0003],[118,0.0082],[119,0.0051],[120,0.0121],[121,0.0089],[122,0.0088],[123,0.0006],[124,0.0044],[125,0.0002],[126,0],[127,0.0045],[128,0],[131,0],[132,0.0002],[133,0.0054],[134,0.0152],[135,0.02],[136,0.0142],[137,0.033],[138,0.0366],[139,0.0002],[140,0.0318],[141,0.0224],[142,0.0259],[143,0.0343],[144,0.0197],[145,0.0134],[146,0.0133],[147,0.0258],[148,0.0136],[149,0.0095],[150,0.0181],[151,0.0231],[152,0.015],[153,0.0189],[154,0.0142],[155,0.022],[156,0.0097],[157,0.0154],[158,0.0155],[159,0.0242],[160,0.0103],[161,0.0162],[162,0.03],[163,0.019],[164,0.0163],[165,0.0204],[166,0.0205],[167,0.0294],[168,0.0143],[169,0.0323],[170,0.0295],[171,0.0393],[172,0.0374],[173,0.0517],[174,0.1959],[175,0.2481],[176,0.2972],[177,0.2878],[178,0.2614],[179,0.2931],[180,0.3143],[181,0.3284],[182,0.3305],[183,0.3208],[184,0.2843],[185,0.2954],[186,0.3161],[187,0.3144],[188,0.3143],[189,0.3388],[190,0.3553],[191,0.3824],[192,0.4305],[193,0.3437],[194,0.4362],[195,0.4021],[196,0.521],[197,0.6728],[198,0.7382],[199,0.7675],[200,0.7704],[201,0.6889],[202,0.7158],[203,0.6844],[204,0.677],[205,0.7053],[206,0.6893],[207,0.692],[208,0.6563],[209,0.6301],[210,0.5412],[211,0.5159],[212,0.5396],[213,0.5426],[214,0.5074],[215,0.523],[216,0.5243],[217,0.4991],[218,0.5067],[219,0.4787],[220,0.4874],[221,0.4751],[222,0.4365],[223,0.48],[224,0.4612],[225,0.4874],[226,0.48],[227,0.4807],[228,0.4985],[229,0.5144],[230,0.5482],[231,0.5571],[232,0.5557],[233,0.6364],[234,0.6608],[235,0.6342],[236,0.6464],[237,0.6194],[238,0.6097],[239,0.6551],[240,0.5986],[241,0.6814],[242,0.7051],[243,0.7241],[244,0.7165],[245,0.7234],[246,0.7974],[247,0.7859],[248,0.6948],[249,0.6541],[250,0.5937],[251,0.5751],[252,0.5163],[253,0.5094],[254,0.5145],[255,0.5321],[256,0.4786],[257,0.5108],[258,0.5839],[259,0.5758],[260,0.5246],[261,0.5373],[262,0.4419],[263,0.2691],[264,0.2935],[265,0.323],[266,0.3156],[267,0.2744],[268,0.2612],[269,0.2906],[270,0.2963],[271,0.3238],[272,0.3612],[273,0.3966],[275,0.4014],[277,0.3795],[278,0.3723],[279,0.3718],[280,0.3562],[281,0.3554],[282,0.3734],[283,0.3629],[284,0.361],[285,0.3645],[286,0.3739],[288,0.3808],[289,0.3897],[290,0.3932],[291,0.366],[292,0.3349],[293,0.3136],[294,0.2884],[295,0.3493],[296,0.4522],[297,0.4733],[298,0.6196],[299,0.6534],[300,0.6063],[301,0.5852],[302,0.4175],[303,0.3789],[304,0.2559],[305,0.1983],[306,0.1678],[307,0.1621],[308,0.1302],[309,0.137],[310,0.1497],[311,0.1383],[312,0.1185],[313,0.0875],[314,0.0515],[315,0.0512],[316,0.0481],[317,0.0312],[318,0.0427],[319,0.0394],[320,0.0837],[321,0.0707],[322,0.0537],[323,0.0757],[324,0.0736],[325,0.0471],[326,0.1002],[327,0.0906],[328,0.091],[329,0.1133],[330,0.1676],[331,0.152],[332,0.0906],[333,0.1238],[334,0.1013],[335,0.1288],[336,0.0758],[337,0.0531],[338,0.0896],[339,0.1032],[340,0.1243],[341,0.1026],[342,0.1213],[343,0.1144],[344,0.0817],[345,0.1229],[346,0.1347],[347,0.1566],[348,0.1478],[349,0.124],[350,0.1382],[351,0.1214],[352,0.1111],[353,0.1403],[354,0.1214],[355,0.1132],[356,0.1127],[357,0.1112],[358,0.0741],[359,0.098],[360,0.098]]
       }
       CTRL_L_mouth_lipsBlow {type=MovieSceneScriptingFloatChannel, default=0, const [0,492]}
       CTRL_R_mouth_lipsTogetherD {
        type=MovieSceneScriptingFloatChannel
        default=0
        thinned_from=492
        keys [[-131,0.1887],[-130,0.2102],[-129,0.2099],[-128,0.2054],[-127,0.1957],[-126,0.1991],[-125,0.1924],[-124,0.1844],[-123,0.1811],[-122,0.1806],[-121,0.1769],[-120,0.1781],[-119,0.1768],[-118,0.1806],[-117,0.182],[-116,0.1878],[-115,0.1758],[-113,0.1497],[-112,0.1346],[-111,0.1235],[-110,0.1329],[-109,0.1506],[-108,0.1553],[-107,0.1365],[-106,0.134],[-105,0.1446],[-104,0.1629],[-103,0.1785],[-102,0.1639],[-101,0.1517],[-100,0.1686],[-99,0.1791],[-98,0.1686],[-97,0.1635],[-96,0.173],[-95,0.176],[-94,0.1719],[-93,0.1723],[-92,0.1842],[-91,0.2003],[-89,0.2581],[-88,0.2752],[-87,0.2846],[-86,0.2987],[-85,0.3087],[-84,0.3215],[-83,0.333],[-82,0.3295],[-81,0.3059],[-80,0.2252],[-79,0.2157],[-78,0.1315],[-77,0.0374],[-76,0.0087],[-75,0.0089],[-74,0.0087],[-73,0.011],[-72,0.011],[-71,0.0099],[-70,0.0107],[-69,0.01],[-68,0.01],[-67,0.0123],[-66,0.0108],[-65,0.0114],[-63,0.0119],[-62,0.0138],[-61,0.0149],[-60,0.0196],[-59,0.0295],[-58,0.0333],[-57,0.0353],[-56,0.0412],[-55,0.0437],[-54,0.0433],[-53,0.047],[-52,0.0489],[-51,0.0484],[-50,0.0496],[-49,0.05],[-48,0.0424],[-47,0.0424],[-46,0.0413],[-45,0.0313],[-44,0.0417],[-43,0.0775],[-42,0.1161],[-41,0.1526],[-40,0.1736],[-39,0.1354],[-38,0.1142],[-37,0.0099],[-36,0.0094],[-35,0.0108],[-34,0.0115],[-33,0.0172],[-32,0.0276],[-31,0.0289],[-30,0.035],[-29,0.0329],[-28,0.0361],[-27,0.0419],[-26,0.0383],[-24,0.0363],[-23,0.0254],[-22,0.035],[-21,0.0272],[-20,0.0352],[-19,0.0347],[-18,0.0317],[-17,0.0331],[-16,0.0373],[-15,0.0368],[-14,0.0383],[-13,0.0404],[-12,0.0388],[-11,0.0443],[-10,0.0443],[-9,0.0428],[-8,0.0457],[-7,0.0407],[-6,0.0453],[-5,0.0405],[-4,0.0498],[-3,0.0454],[-2,0.0441],[-1,0.0488],[0,0.0433],[1,0.0453],[2,0.045],[3,0.0403],[4,0.0442],[5,0.0468],[6,0.0518],[7,0.0482],[8,0.0492],[9,0.051],[10,0.049],[11,0.0472],[12,0.0465],[13,0.0473],[14,0.0455],[15,0.0477],[16,0.0462],[17,0.0461],[18,0.0486],[19,0.0473],[21,0.0471],[22,0.0434],[23,0.0431],[24,0.0423],[25,0.0421],[26,0.0432],[27,0.0459],[28,0.0455],[29,0.0454],[30,0.0444],[31,0.0475],[32,0.0442],[33,0.0398],[34,0.0441],[35,0.0436],[36,0.0399],[37,0.042],[38,0.047],[39,0.0495],[40,0.0456],[41,0.0482],[42,0.0487],[43,0.0505],[45,0.051],[46,0.0478],[47,0.0463],[48,0.0422],[49,0.0375],[50,0.0353],[51,0.0371],[52,0.0436],[53,0.0346],[54,0.0379],[55,0.0325],[56,0.0295],[57,0.0158],[58,0.0027],[59,0.0114],[60,0.0143],[61,0.0151],[62,0.0184],[63,0.0184],[64,0.0192],[65,0.0192],[66,0.0267],[67,0.0126],[68,0.0149],[69,0.0235],[70,0.016],[71,0.0207],[72,0.0166],[73,0.0068],[74,0.0119],[75,0.0277],[76,0.0196],[77,0.024],[78,0.0232],[79,0.0094],[80,0.01],[81,0.0191],[82,0.0105],[84,0.0098],[85,0.0104],[86,0.0158],[87,0.0144],[88,0.0002],[89,0.0081],[90,0.0052],[91,0.0051],[92,0.0052],[93,0.0036],[94,0],[97,0],[98,0.0049],[99,0.0152],[100,0.0171],[101,0.0081],[102,0.0047],[103,0],[105,0],[106,0.0026],[107,0.0098],[108,0.0006],[109,0.0094],[110,0.0048],[111,0.0191],[112,0.0096],[113,0.0083],[114,0.0003],[115,0.0135],[116,0.009],[117,0.0005],[118,0.0123],[119,0.0056],[120,0.0121],[121,0.0134],[122,0.0132],[123,0.0046],[124,0.0044],[125,0.0002],[126,0],[127,0.0045],[128,0.0001],[129,0.0049],[130,0.005],[131,0.0008],[132,0.0002],[133,0.0054],[134,0.0245],[135,0.0151],[136,0.01],[137,0.033],[138,0.0418],[139,0.0003],[140,0.0425],[141,0.028],[142,0.0311],[143,0.0324],[144,0.0243],[145,0.0177],[146,0.0134],[147,0.0257],[148,0.0092],[149,0.0138],[150,0.0182],[151,0.0234],[152,0.02],[153,0.0237],[154,0.0189],[155,0.0222],[156,0.0098],[157,0.0204],[158,0.0207],[159,0.0339],[160,0.0153],[161,0.0212],[162,0.0352],[163,0.0238],[164,0.0213],[165,0.0255],[166,0.0252],[167,0.0294],[168,0.0224],[169,0.0305],[171,0.0327],[172,0.0242],[173,0.0124],[174,0.0065],[175,0.0115],[176,0.0141],[177,0.0163],[178,0.0202],[179,0.0178],[180,0.0295],[181,0.0624],[182,0.0641],[183,0.0628],[184,0.0364],[185,0.0612],[186,0.098],[187,0.1035],[188,0.1039],[189,0.1112],[190,0.1125],[191,0.0929],[192,0.0616],[193,0.0318],[194,0.0331],[195,0.0289],[196,0.0691],[197,0.1183],[198,0.1064],[199,0.1059],[200,0.1111],[201,0.1904],[202,0.1261],[203,0.1228],[204,0.1131],[205,0.0949],[206,0.0941],[207,0.1221],[208,0.0964],[209,0.0736],[210,0.0564],[211,0.0322],[212,0.0392],[213,0.0307],[214,0.0201],[215,0.0259],[216,0.0332],[217,0.0255],[218,0.0347],[219,0.0347],[220,0.034],[221,0.023],[222,0.0242],[223,0.043],[224,0.0323],[225,0.0275],[226,0.0329],[227,0.0409],[228,0.0218],[229,0.0497],[230,0.0722],[231,0.0978],[232,0.1202],[233,0.184],[234,0.2014],[235,0.1721],[236,0.0436],[237,0.0232],[238,0.047],[239,0.0602],[240,0.044],[241,0.0852],[242,0.0972],[243,0.1273],[244,0.1775],[245,0.2339],[246,0.2556],[247,0.2358],[248,0.1587],[249,0.1224],[250,0.141],[251,0.1619],[252,0.2406],[253,0.2917],[254,0.2786],[255,0.258],[256,0.2715],[257,0.2408],[258,0.1309],[259,0.1381],[260,0.2005],[261,0.2027],[262,0.1756],[263,0.2101],[264,0.228],[265,0.2219],[266,0.2238],[267,0.1933],[268,0.1484],[269,0.1999],[270,0.2123],[271,0.2202],[272,0.2221],[273,0.2287],[274,0.247],[275,0.2579],[276,0.2469],[277,0.2435],[278,0.2409],[279,0.2262],[280,0.2328],[281,0.2355],[282,0.2181],[283,0.2329],[284,0.2352],[285,0.2259],[286,0.2524],[287,0.2474],[288,0.245],[289,0.2401],[290,0.2373],[291,0.2335],[292,0.208],[293,0.181],[294,0.1797],[295,0.2168],[296,0.2337],[297,0.2344],[298,0.2097],[299,0.1922],[300,0.1312],[301,0.1209],[302,0.0385],[303,0.0349],[304,0.0255],[305,0.0248],[307,0.0287],[308,0.0272],[309,0.028],[310,0.0284],[311,0.0245],[312,0.0245],[313,0.0218],[314,0.0277],[315,0.0273],[316,0.0262],[317,0.0234],[318,0.02],[319,0.0273],[320,0.0262],[321,0.026],[322,0.0156],[323,0.0192],[324,0.0291],[325,0.0302],[326,0.0255],[327,0.0219],[328,0.0253],[329,0.0249],[330,0.0239],[331,0.0271],[332,0.029],[333,0.0247],[334,0.0244],[335,0.0207],[336,0.0253],[337,0.0225],[338,0.0284],[339,0.0216],[340,0.0241],[341,0.0178],[342,0.0209],[343,0.0214],[344,0.0217],[347,0.02],[348,0.0175],[349,0.0195],[350,0.0141],[351,0.0242],[352,0.0181],[353,0.0218],[354,0.0201],[355,0.0226],[356,0.0246],[357,0.0212],[358,0.0221],[359,0.0212],[360,0.0212]]
       }
       CTRL_C_tongue_roll.X {type=MovieSceneScriptingFloatChannel, default=0, const [0,492]}
       CTRL_C_tongue_roll.Y {type=MovieSceneScriptingFloatChannel, default=0, const [0,492]}
       CTRL_C_tongue.X {type=MovieSceneScriptingFloatChannel, default=0, const [0,492]}
       CTRL_C_tongue.Y {type=MovieSceneScriptingFloatChannel, default=0, const [0,492]}
       CTRL_C_tongue_narrowWide {type=MovieSceneScriptingFloatChannel, default=0, const [0,492]}
       CTRL_C_tongue_inOut {type=MovieSceneScriptingFloatChannel, default=0, const [0,492]}
       CTRL_R_nose.X {
        type=MovieSceneScriptingFloatChannel
        default=0
        thinned_from=492
        keys [[-131,0.0007],[-128,0.0065],[-127,0.0074],[-126,0.0077],[-125,0.007],[-124,0.0075],[-122,0.0072],[-121,0.0065],[-120,0.0065],[-119,0.0062],[-118,0.0056],[-117,0.0056],[-115,0.006],[-114,0.0058],[-112,0.006],[-111,0.0069],[-110,0.0072],[-109,0.0077],[-108,0.0098],[-107,0.0114],[-106,0.0116],[-104,0.0126],[-102,0.0147],[-101,0.0162],[-100,0.0164],[-99,0.0158],[-98,0.0159],[-96,0.017],[-95,0.0171],[-94,0.0141],[-92,0.0137],[-91,0.0127],[-90,0.0105],[-89,0.0059],[-88,0.0007],[-86,-0.0113],[-85,-0.0177],[-84,-0.0245],[-83,-0.032],[-82,-0.0401],[-81,-0.0448],[-80,-0.0538],[-79,-0.0557],[-78,-0.0718],[-77,-0.0862],[-76,-0.0957],[-75,-0.1093],[-74,-0.1136],[-73,-0.1109],[-71,-0.1168],[-70,-0.1255],[-69,-0.1268],[-68,-0.1262],[-67,-0.1274],[-66,-0.1256],[-65,-0.1248],[-63,-0.1199],[-62,-0.1152],[-61,-0.1033],[-60,-0.0859],[-59,-0.0577],[-58,-0.0367],[-57,-0.0312],[-56,-0.0194],[-55,-0.0145],[-54,-0.0118],[-53,-0.0099],[-52,-0.0091],[-51,-0.0055],[-50,-0.0051],[-49,-0.0051],[-48,-0.004],[-47,-0.0055],[-46,-0.006],[-45,-0.0093],[-44,-0.0102],[-43,-0.0131],[-42,-0.0181],[-41,-0.0261],[-40,-0.0378],[-39,-0.055],[-38,-0.0603],[-37,-0.091],[-36,-0.1014],[-35,-0.1159],[-34,-0.114],[-33,-0.1008],[-32,-0.0688],[-31,-0.0611],[-30,-0.0279],[-29,-0.0245],[-28,-0.017],[-27,-0.013],[-26,-0.011],[-25,-0.0087],[-24,-0.0078],[-23,0.0014],[-22,0.0058],[-21,0.0127],[-20,0.0159],[-19,0.0179],[-18,0.0184],[-17,0.0214],[-16,0.0232],[-15,0.0233],[-14,0.0239],[-13,0.0237],[-12,0.023],[-11,0.0227],[-10,0.0218],[-9,0.0223],[-8,0.0212],[-7,0.0209],[-6,0.0211],[-5,0.0196],[-4,0.02],[-3,0.0198],[-2,0.0201],[-1,0.0198],[0,0.0187],[2,0.0178],[3,0.0162],[4,0.0144],[5,0.013],[6,0.0095],[7,0.0086],[8,0.0082],[10,0.0055],[11,0.0048],[12,0.0044],[13,0.0037],[14,0.0032],[15,0.002],[16,0.0004],[18,-0.001],[19,-0.002],[22,-0.0031],[25,-0.0027],[27,-0.0034],[29,-0.0051],[31,-0.0053],[32,-0.0047],[33,-0.0055],[36,-0.0056],[37,-0.0063],[40,-0.0062],[41,-0.0048],[42,-0.0043],[43,-0.004],[44,-0.0027],[45,-0.0022],[46,0.0046],[47,0.0043],[48,0.0079],[50,0.0104],[51,0.0131],[53,0.0159],[54,0.0148],[55,0.0144],[56,0.0142],[58,0.0116],[59,0.0106],[60,0.0104],[61,0.01],[62,0.0089],[65,0.0088],[66,0.0083],[67,0.0085],[68,0.0081],[69,0.0064],[70,0.0051],[71,0.0046],[72,0.0037],[73,0.0035],[74,0.002],[75,0.001],[76,-0.0004],[77,-0.0007],[79,-0.0009],[80,-0.0018],[82,-0.0029],[84,-0.0013],[85,-0.0016],[87,-0.0015],[88,-0.001],[89,-0.0008],[90,-0.0016],[91,-0.0013],[95,-0.0012],[98,-0.0028],[99,-0.003],[100,-0.0023],[101,-0.0019],[102,-0.0019],[103,-0.0016],[104,0],[105,0],[106,0.0008],[108,0.0016],[110,0.0019],[111,0.0024],[112,0.0037],[113,0.0054],[114,0.0066],[115,0.0068],[116,0.0073],[117,0.0092],[118,0.013],[119,0.0148],[120,0.0158],[121,0.0157],[123,0.0135],[124,0.0116],[125,0.0116],[126,0.0103],[127,0.01],[128,0.0094],[129,0.0096],[131,0.0067],[133,0.0067],[134,0.0103],[135,0.0121],[136,0.0125],[137,0.0136],[138,0.0092],[139,0.0016],[140,-0.0036],[141,-0.0053],[142,-0.0067],[143,-0.0068],[145,-0.0055],[146,-0.0047],[147,-0.0053],[148,-0.0073],[149,-0.0061],[150,-0.0055],[152,-0.0067],[153,-0.0071],[154,-0.0079],[155,-0.0071],[156,-0.0065],[157,-0.0069],[158,-0.0079],[159,-0.0086],[160,-0.0085],[161,-0.0086],[162,-0.008],[164,-0.0062],[166,-0.0053],[167,-0.0058],[168,-0.0055],[169,-0.0041],[170,-0.0042],[171,-0.0032],[172,-0.0026],[174,-0.0003],[176,0.004],[177,0.0043],[179,0.0063],[180,0.0069],[181,0.0071],[182,0.008],[184,0.0088],[185,0.0099],[188,0.0104],[190,0.0127],[191,0.0127],[192,0.0187],[193,0.0227],[194,0.0253],[195,0.0247],[196,0.0267],[197,0.0332],[198,0.0405],[199,0.0452],[200,0.0463],[201,0.0476],[202,0.0435],[203,0.0425],[204,0.0427],[205,0.0435],[206,0.0438],[207,0.043],[208,0.0415],[209,0.0406],[210,0.0378],[211,0.0363],[212,0.0366],[213,0.0359],[214,0.0358],[215,0.0347],[216,0.0334],[217,0.0333],[218,0.0327],[219,0.0314],[220,0.0313],[221,0.0308],[222,0.0295],[223,0.0289],[225,0.0292],[226,0.0298],[227,0.0295],[228,0.03],[230,0.0321],[231,0.0322],[232,0.0331],[233,0.0342],[234,0.0342],[236,0.0298],[237,0.0277],[238,0.0266],[239,0.0206],[240,0.0157],[241,0.0127],[242,0.0123],[243,0.0102],[244,0.0046],[245,0.0028],[246,-0.0026],[247,-0.0112],[248,-0.0151],[249,-0.0125],[250,-0.0079],[251,-0.0092],[252,-0.0079],[253,-0.01],[254,-0.0106],[255,-0.0104],[256,-0.0098],[257,-0.0044],[258,0.0007],[259,0.0011],[260,0.0043],[261,0.0052],[262,0.0058],[263,0.0031],[265,0.0014],[266,0.0026],[267,0.0018],[268,0.0035],[269,0.0042],[270,0.0042],[271,0.0008],[272,0.0037],[273,0.0044],[274,0.0085],[275,0.009],[276,0.0089],[277,0.0085],[279,0.0073],[280,0.0065],[282,0.0067],[283,0.0063],[285,0.0061],[286,0.0066],[288,0.0062],[289,0.0067],[290,0.0067],[291,0.007],[292,0.0095],[293,0.0113],[294,0.0126],[295,0.0112],[296,0.0087],[297,0.0084],[298,0.0097],[299,0.013],[300,0.0199],[301,0.0202],[302,0.0214],[304,0.021],[305,0.0211],[306,0.0206],[308,0.0208],[310,0.0205],[311,0.0184],[313,0.0178],[314,0.0172],[315,0.0175],[316,0.0179],[317,0.0178],[318,0.0182],[319,0.0183],[320,0.017],[321,0.0177],[323,0.0169],[324,0.0167],[325,0.0171],[326,0.0158],[328,0.0144],[330,0.0145],[332,0.0139],[334,0.0136],[336,0.0129],[337,0.0128],[338,0.0122],[340,0.0118],[342,0.0125],[343,0.0126],[344,0.0132],[345,0.0125],[346,0.0121],[347,0.0129],[348,0.0127],[349,0.0117],[350,0.0117],[351,0.0112],[352,0.011],[353,0.0095],[356,0.0092],[357,0.0095],[358,0.0096],[359,0.0102],[360,0.0102]]
       }
       CTRL_R_nose.Y {
        type=MovieSceneScriptingFloatChannel
        default=0
        thinned_from=492
        keys [[-131,0.0098],[-130,0.0059],[-129,0.0041],[-128,0.0048],[-127,0.0007],[-126,-0.0001],[-125,-0.0013],[-124,-0.0019],[-123,-0.0013],[-122,-0.0012],[-121,-0.0044],[-120,-0.004],[-119,-0.0048],[-118,-0.0047],[-117,-0.0036],[-116,0.0042],[-115,0.0065],[-113,-0.0014],[-112,-0.0006],[-111,0.0009],[-110,0.0011],[-108,0.0107],[-107,0.0115],[-106,0.0117],[-105,0.0155],[-104,0.0162],[-103,0.0192],[-102,0.0166],[-101,0.0095],[-100,0.0128],[-99,0.0097],[-98,0.01],[-97,0.0097],[-96,0.0068],[-95,0.0059],[-94,0.0072],[-93,0.0073],[-92,0.0088],[-91,0.0105],[-90,0.0143],[-89,0.0241],[-88,0.0336],[-87,0.0428],[-86,0.0565],[-84,0.086],[-82,0.1028],[-81,0.1062],[-80,0.1143],[-79,0.1142],[-78,0.1133],[-77,0.1026],[-76,0.0922],[-75,0.0822],[-74,0.0672],[-73,0.058],[-72,0.0474],[-71,0.0466],[-69,0.0357],[-68,0.0368],[-66,0.0384],[-65,0.0389],[-64,0.0398],[-63,0.0385],[-61,0.0348],[-60,0.0363],[-59,0.0355],[-58,0.0365],[-56,0.0361],[-55,0.0375],[-54,0.0375],[-53,0.0371],[-52,0.0371],[-51,0.0273],[-50,0.0251],[-49,0.025],[-48,0.0223],[-47,0.0224],[-46,0.0215],[-45,0.0215],[-44,0.0221],[-43,0.024],[-42,0.0303],[-41,0.0427],[-40,0.0631],[-39,0.0785],[-38,0.0811],[-37,0.0833],[-36,0.0748],[-35,0.0712],[-34,0.0732],[-33,0.0734],[-32,0.0637],[-31,0.0595],[-30,0.039],[-29,0.0367],[-28,0.0335],[-27,0.0361],[-26,0.0379],[-25,0.0359],[-24,0.0346],[-23,0.011],[-22,0.0046],[-21,-0.0142],[-20,-0.0066],[-19,-0.0063],[-18,-0.0036],[-17,-0.0013],[-16,0.0061],[-15,0.0112],[-14,0.0155],[-13,0.0194],[-12,0.0253],[-11,0.0264],[-10,0.0299],[-9,0.0278],[-8,0.0303],[-5,0.0366],[-4,0.0355],[-3,0.0381],[-2,0.0371],[-1,0.0377],[0,0.0404],[1,0.0406],[3,0.0391],[4,0.0396],[5,0.0383],[6,0.0386],[7,0.0383],[8,0.0378],[9,0.0358],[10,0.0345],[11,0.0364],[12,0.0365],[13,0.037],[14,0.0371],[15,0.0369],[16,0.0364],[17,0.0361],[18,0.0388],[19,0.0391],[20,0.04],[21,0.0411],[22,0.0425],[23,0.0426],[24,0.0414],[25,0.0378],[26,0.0387],[27,0.0394],[28,0.0405],[29,0.0423],[30,0.043],[31,0.042],[32,0.039],[33,0.0403],[34,0.0428],[35,0.0422],[36,0.0386],[37,0.04],[38,0.0422],[39,0.0426],[40,0.0404],[41,0.0376],[42,0.0372],[43,0.0382],[45,0.0368],[46,0.0285],[47,0.0298],[48,0.0231],[49,0.0195],[50,0.0132],[51,0.0043],[52,0.0028],[53,0.0035],[54,0.0039],[55,0.0026],[56,-0.0028],[57,-0.0059],[58,-0.0079],[59,-0.008],[60,-0.0078],[62,-0.0054],[63,-0.0052],[64,-0.0025],[65,-0.0025],[66,0.0011],[67,-0.0002],[68,-0.0016],[69,-0.001],[72,0.0035],[73,0.0035],[74,0.0059],[75,0.0059],[76,0.0075],[77,0.0097],[78,0.0096],[79,0.0076],[80,0.0106],[83,0.0098],[84,0.0061],[85,0.0071],[86,0.0099],[87,0.0099],[88,0.0113],[89,0.0082],[90,0.0106],[91,0.0078],[92,0.0108],[93,0.0109],[94,0.0113],[95,0.011],[96,0.0115],[97,0.0141],[99,0.0164],[100,0.0186],[101,0.0179],[102,0.0176],[103,0.0191],[104,0.0159],[105,0.0157],[106,0.0146],[108,0.0145],[109,0.0127],[110,0.0091],[111,0.0088],[112,0.0052],[113,-0.0032],[114,-0.01],[115,-0.0137],[116,-0.0192],[117,-0.0309],[118,-0.0553],[119,-0.0694],[120,-0.0798],[121,-0.0823],[122,-0.0819],[123,-0.0874],[124,-0.0825],[125,-0.0823],[126,-0.0812],[127,-0.0811],[128,-0.0784],[130,-0.0697],[131,-0.0706],[132,-0.069],[133,-0.0669],[134,-0.0781],[135,-0.093],[136,-0.1056],[137,-0.0934],[138,-0.0563],[139,-0.0349],[140,-0.0256],[141,-0.0265],[142,-0.0211],[143,-0.0192],[144,-0.0191],[145,-0.0234],[146,-0.0295],[147,-0.0304],[148,-0.0251],[149,-0.025],[150,-0.0263],[151,-0.026],[152,-0.0286],[153,-0.0269],[154,-0.0195],[155,-0.0195],[157,-0.0177],[158,-0.0135],[160,-0.0106],[161,-0.0042],[162,0.0011],[163,0.0032],[164,0.0011],[165,0.0039],[166,0.0107],[167,0.0146],[168,0.0141],[169,0.0148],[170,0.0186],[171,0.0201],[172,0.017],[174,0.0026],[175,-0.0063],[176,-0.0098],[177,-0.0093],[178,-0.008],[179,-0.0096],[180,-0.0096],[181,-0.0117],[182,-0.0162],[183,-0.0192],[184,-0.0227],[185,-0.0244],[186,-0.0243],[187,-0.023],[188,-0.0229],[189,-0.0286],[190,-0.0309],[191,-0.0355],[192,-0.0819],[193,-0.1142],[194,-0.1502],[195,-0.1474],[196,-0.1612],[197,-0.1977],[198,-0.2289],[199,-0.2388],[200,-0.2397],[201,-0.233],[202,-0.2195],[203,-0.1911],[204,-0.1817],[205,-0.1714],[206,-0.1659],[207,-0.153],[208,-0.143],[209,-0.1351],[210,-0.1163],[211,-0.1075],[212,-0.1085],[213,-0.1079],[214,-0.1093],[215,-0.1031],[216,-0.0984],[217,-0.0944],[218,-0.0923],[219,-0.0889],[220,-0.0837],[221,-0.0822],[222,-0.0784],[223,-0.0756],[224,-0.0754],[225,-0.0745],[226,-0.073],[227,-0.0726],[228,-0.0749],[229,-0.0778],[230,-0.0758],[231,-0.0727],[232,-0.0761],[233,-0.09],[234,-0.0966],[235,-0.0943],[236,-0.0711],[237,-0.0665],[238,-0.0565],[239,-0.0522],[240,-0.0401],[241,-0.0292],[242,-0.0277],[243,-0.0271],[244,-0.0172],[245,-0.0109],[246,-0.0091],[247,-0.0039],[248,-0.0047],[249,-0.0029],[250,-0.007],[251,0.0004],[252,0.004],[253,0.0132],[254,0.0184],[255,0.0226],[256,0.0224],[257,0.0113],[258,0.0105],[259,0.0126],[260,0.0416],[261,0.0737],[262,0.088],[263,0.109],[264,0.1109],[265,0.1077],[266,0.1037],[267,0.0979],[268,0.084],[269,0.0528],[270,0.0471],[271,0.0379],[272,0.0149],[273,0.0027],[274,-0.0074],[275,-0.0091],[276,-0.0101],[277,-0.0045],[278,-0.0007],[279,0],[280,0.0005],[281,0.0003],[283,-0.0016],[286,-0.0016],[287,0.001],[288,0.0046],[289,0.0066],[290,0.0057],[291,0.0036],[292,0.0085],[293,0.0147],[294,0.02],[295,0.0206],[296,0.0166],[297,0.0156],[298,0.0048],[299,-0.0002],[300,-0.0134],[301,-0.0136],[302,-0.0134],[303,-0.014],[304,-0.0158],[305,-0.0161],[306,-0.0146],[307,-0.0132],[309,-0.01],[310,-0.0095],[311,-0.0049],[313,-0.0036],[314,-0.001],[315,-0.0049],[316,-0.0046],[318,-0.0029],[319,-0.0027],[321,-0.0013],[322,-0.0016],[323,-0.003],[324,0.0002],[325,0.002],[326,0.003],[327,0.0046],[328,0.0074],[329,0.0089],[330,0.0082],[331,0.0107],[332,0.0123],[333,0.012],[334,0.0126],[335,0.0138],[336,0.0132],[337,0.0134],[338,0.0141],[339,0.013],[340,0.0123],[341,0.0125],[345,0.0104],[346,0.0094],[347,0.0067],[348,0.0057],[349,0.0062],[350,0.0075],[351,0.0093],[353,0.0105],[354,0.0108],[355,0.0123],[356,0.0126],[357,0.0116],[358,0.0136],[359,0.0126],[360,0.0126]]
       }
       CTRL_L_nose.X {
        type=MovieSceneScriptingFloatChannel
        default=0
        thinned_from=492
        keys [[-131,-0.012],[-130,-0.0106],[-129,-0.0081],[-128,-0.006],[-127,-0.005],[-126,-0.0047],[-125,-0.0056],[-124,-0.005],[-123,-0.0054],[-122,-0.0056],[-121,-0.0067],[-120,-0.0062],[-119,-0.0067],[-118,-0.0077],[-117,-0.0077],[-115,-0.0072],[-114,-0.0074],[-113,-0.0071],[-112,-0.007],[-111,-0.0063],[-109,-0.0056],[-107,-0.0016],[-105,-0.0009],[-103,0.0013],[-102,0.0026],[-101,0.0043],[-100,0.0041],[-99,0.0032],[-98,0.0033],[-96,0.0043],[-95,0.0044],[-94,0.0019],[-93,0.0018],[-92,0.0023],[-91,0.001],[-90,-0.0018],[-89,-0.0071],[-88,-0.0128],[-87,-0.019],[-86,-0.0249],[-85,-0.0315],[-84,-0.0389],[-83,-0.0467],[-82,-0.0536],[-81,-0.0581],[-80,-0.0674],[-79,-0.0693],[-78,-0.0865],[-77,-0.1024],[-76,-0.1138],[-75,-0.1305],[-74,-0.136],[-73,-0.1347],[-72,-0.1377],[-71,-0.1416],[-70,-0.1504],[-69,-0.1523],[-68,-0.1514],[-67,-0.1533],[-66,-0.151],[-65,-0.1502],[-63,-0.1452],[-62,-0.1402],[-61,-0.1277],[-60,-0.1085],[-59,-0.0761],[-58,-0.0515],[-57,-0.0453],[-56,-0.032],[-55,-0.0265],[-54,-0.0234],[-53,-0.0214],[-52,-0.0206],[-51,-0.0164],[-50,-0.0161],[-49,-0.0161],[-48,-0.0143],[-47,-0.0157],[-46,-0.0164],[-45,-0.0203],[-44,-0.0213],[-43,-0.0247],[-42,-0.0301],[-41,-0.0383],[-40,-0.0513],[-39,-0.0699],[-38,-0.0756],[-37,-0.1109],[-36,-0.1231],[-35,-0.1387],[-34,-0.1368],[-33,-0.1213],[-32,-0.0858],[-31,-0.0771],[-30,-0.0397],[-29,-0.036],[-28,-0.0281],[-27,-0.0239],[-26,-0.0218],[-25,-0.0192],[-24,-0.0183],[-23,-0.0095],[-22,-0.0046],[-21,0.0021],[-20,0.0047],[-19,0.0069],[-18,0.0076],[-17,0.0106],[-16,0.0124],[-15,0.0125],[-14,0.0131],[-11,0.0127],[-10,0.0119],[-9,0.0125],[-8,0.0114],[-6,0.0116],[-5,0.01],[-4,0.0108],[-3,0.0102],[-2,0.0105],[-1,0.0101],[0,0.009],[2,0.0081],[3,0.0069],[4,0.0048],[5,0.0033],[6,-0.0002],[8,-0.0013],[9,-0.0027],[10,-0.0038],[13,-0.0062],[14,-0.0069],[16,-0.0092],[17,-0.01],[18,-0.0102],[19,-0.0112],[20,-0.0117],[22,-0.0123],[25,-0.0125],[27,-0.0132],[29,-0.0148],[31,-0.0145],[32,-0.014],[33,-0.0147],[34,-0.015],[36,-0.0147],[37,-0.0155],[40,-0.0154],[41,-0.014],[42,-0.0133],[43,-0.0132],[44,-0.0119],[45,-0.0114],[46,-0.0049],[47,-0.0051],[48,-0.0012],[49,-0.0002],[50,0.0012],[51,0.0042],[53,0.0071],[54,0.0058],[55,0.0047],[56,0.0046],[58,0.0014],[59,0],[61,-0.0004],[62,-0.0014],[63,-0.0015],[64,-0.0018],[65,-0.0018],[66,-0.0027],[67,-0.0016],[68,-0.0023],[69,-0.0037],[70,-0.0053],[73,-0.007],[74,-0.0089],[75,-0.01],[76,-0.0119],[77,-0.0116],[79,-0.0117],[81,-0.0131],[82,-0.0136],[84,-0.0121],[85,-0.0121],[86,-0.0109],[87,-0.0111],[88,-0.0109],[89,-0.0111],[90,-0.0119],[95,-0.0116],[97,-0.0123],[98,-0.0125],[99,-0.0124],[101,-0.0115],[102,-0.0114],[103,-0.0116],[104,-0.0101],[105,-0.0101],[106,-0.0087],[107,-0.0085],[108,-0.0079],[109,-0.0078],[110,-0.0081],[111,-0.0077],[112,-0.0065],[113,-0.0048],[114,-0.0034],[116,-0.0028],[117,-0.0008],[118,0.0021],[119,0.0035],[120,0.0042],[121,0.0045],[123,0.002],[124,0.0005],[125,0.0007],[126,-0.001],[127,-0.001],[128,-0.0017],[129,-0.0014],[130,-0.0017],[131,-0.0026],[132,-0.0033],[133,-0.0029],[134,-0.0004],[135,0.0013],[136,0.0007],[137,0.0019],[138,-0.0015],[139,-0.0084],[140,-0.0129],[141,-0.0146],[142,-0.0155],[143,-0.0152],[146,-0.0132],[147,-0.0144],[148,-0.0162],[150,-0.0144],[151,-0.0148],[152,-0.0154],[153,-0.0156],[154,-0.0163],[156,-0.015],[157,-0.0153],[158,-0.0162],[159,-0.0169],[160,-0.0168],[161,-0.0171],[163,-0.0157],[164,-0.0146],[165,-0.0139],[166,-0.0127],[167,-0.0134],[168,-0.0135],[169,-0.0122],[170,-0.0123],[171,-0.0112],[172,-0.0108],[173,-0.0099],[174,-0.0093],[175,-0.0078],[176,-0.006],[177,-0.0056],[179,-0.0038],[180,-0.0032],[181,-0.0031],[182,-0.0025],[184,-0.0022],[185,-0.0013],[186,-0.0012],[188,-0.0013],[189,-0.0001],[190,0],[191,-0.0002],[192,0.0059],[193,0.011],[194,0.0168],[195,0.0165],[196,0.018],[198,0.0301],[199,0.0336],[201,0.0353],[202,0.0323],[203,0.03],[204,0.0296],[206,0.0304],[207,0.0299],[209,0.0273],[210,0.0248],[211,0.0235],[212,0.0238],[213,0.0232],[214,0.0236],[215,0.0227],[216,0.0216],[217,0.021],[218,0.0207],[219,0.0196],[220,0.019],[222,0.0174],[223,0.017],[225,0.0174],[226,0.018],[227,0.0178],[228,0.0188],[229,0.0201],[231,0.0204],[232,0.0209],[233,0.0217],[234,0.0216],[235,0.0205],[236,0.0202],[238,0.0185],[239,0.0129],[240,0.0071],[241,0.0027],[242,0.002],[243,-0.0016],[244,-0.0082],[245,-0.0101],[246,-0.0158],[247,-0.0252],[248,-0.0303],[249,-0.0276],[250,-0.0232],[251,-0.0242],[252,-0.0227],[253,-0.0245],[254,-0.0247],[255,-0.0243],[256,-0.0234],[258,-0.0137],[259,-0.0133],[260,-0.0095],[262,-0.007],[263,-0.0098],[264,-0.0104],[265,-0.0113],[266,-0.0102],[267,-0.0104],[268,-0.0076],[269,-0.007],[270,-0.0071],[271,-0.0112],[272,-0.0088],[273,-0.0084],[274,-0.0048],[275,-0.0045],[277,-0.0044],[278,-0.0046],[279,-0.0051],[280,-0.006],[281,-0.0059],[282,-0.0061],[283,-0.0066],[285,-0.0067],[286,-0.0057],[288,-0.0057],[289,-0.0053],[290,-0.0052],[291,-0.0046],[292,-0.002],[293,0.0003],[294,0.0017],[295,0.0002],[296,-0.0022],[297,-0.0025],[298,-0.0014],[299,0.0024],[300,0.0096],[301,0.0098],[302,0.0103],[307,0.0095],[308,0.0096],[310,0.0092],[311,0.0075],[312,0.0071],[313,0.007],[314,0.0064],[315,0.0062],[316,0.0069],[317,0.0069],[318,0.0073],[319,0.0071],[320,0.006],[321,0.0068],[323,0.0061],[324,0.006],[325,0.0063],[326,0.005],[327,0.0049],[328,0.0044],[332,0.0039],[333,0.0034],[334,0.0036],[335,0.0033],[336,0.0026],[337,0.0027],[338,0.0017],[339,0.0023],[340,0.0016],[342,0.0026],[344,0.0028],[345,0.0024],[346,0.0022],[347,0.0025],[348,0.0022],[349,0.0016],[350,0.0018],[351,0.0013],[352,0.0012],[353,-0.0002],[355,0.0002],[356,0.0002],[357,-0.0002],[358,-0.0003],[359,0.0003],[360,0.0003]]
       }
       CTRL_L_nose.Y {
        type=MovieSceneScriptingFloatChannel
        default=0
        thinned_from=492
        keys [[-131,0.0149],[-130,0.0164],[-129,0.0164],[-128,0.0143],[-127,0.0105],[-126,0.0089],[-125,0.0063],[-124,0.0054],[-123,0.0072],[-122,0.0077],[-121,0.0062],[-120,0.0057],[-119,0.0039],[-118,0.0038],[-117,0.0045],[-116,0.0095],[-115,0.0101],[-114,0.0068],[-113,0.0043],[-112,0.0046],[-110,0.0074],[-109,0.0106],[-108,0.0153],[-107,0.0173],[-106,0.0177],[-105,0.0215],[-104,0.0214],[-103,0.0234],[-102,0.0214],[-101,0.019],[-100,0.0225],[-99,0.0222],[-98,0.0248],[-97,0.0262],[-96,0.0258],[-95,0.0258],[-94,0.0283],[-93,0.0287],[-92,0.0316],[-91,0.036],[-90,0.0413],[-89,0.0499],[-87,0.0667],[-86,0.0744],[-85,0.0838],[-84,0.0938],[-83,0.0978],[-82,0.1005],[-81,0.0982],[-80,0.0902],[-79,0.0897],[-78,0.0844],[-77,0.0718],[-76,0.0666],[-75,0.062],[-74,0.0535],[-73,0.0493],[-72,0.0413],[-71,0.0421],[-70,0.0346],[-69,0.0301],[-68,0.0274],[-67,0.0322],[-66,0.0322],[-65,0.034],[-64,0.034],[-63,0.0288],[-62,0.0294],[-61,0.0293],[-60,0.028],[-59,0.0219],[-58,0.0165],[-57,0.0143],[-56,0.0095],[-55,0.0074],[-54,0.0058],[-53,0.0056],[-52,0.0057],[-51,-0.003],[-50,-0.0024],[-49,-0.0022],[-48,-0.0032],[-47,-0.0002],[-46,0.0004],[-45,0.006],[-44,0.008],[-43,0.0143],[-42,0.0211],[-41,0.0309],[-40,0.0434],[-39,0.054],[-38,0.0562],[-37,0.0632],[-36,0.0614],[-35,0.0586],[-34,0.0616],[-33,0.0626],[-32,0.0483],[-31,0.042],[-30,0.0167],[-29,0.0124],[-28,0.0059],[-26,0.0131],[-24,0.011],[-23,0.0015],[-22,-0.0001],[-21,-0.0099],[-20,-0.0122],[-19,-0.0136],[-18,-0.0138],[-17,-0.0166],[-16,-0.0147],[-15,-0.0135],[-13,-0.0079],[-12,-0.0045],[-11,-0.0034],[-10,-0.0015],[-9,-0.0019],[-8,0.0009],[-7,0.0014],[-6,0.0008],[-5,0.0021],[-4,0.0029],[-3,0.0024],[-2,0.0022],[-1,0.0027],[0,0.0057],[1,0.0062],[2,0.0057],[3,0.0058],[4,0.0063],[5,0.0056],[6,0.007],[7,0.0064],[8,0.0065],[10,0.0058],[11,0.0085],[12,0.01],[13,0.0105],[14,0.0093],[15,0.0077],[16,0.0091],[17,0.0092],[18,0.0098],[19,0.0084],[20,0.009],[21,0.0104],[23,0.0106],[24,0.0097],[25,0.0073],[26,0.0074],[27,0.0073],[28,0.0095],[29,0.0128],[30,0.0145],[31,0.0138],[32,0.0112],[33,0.0118],[34,0.0138],[35,0.0135],[36,0.0117],[37,0.0117],[38,0.0112],[39,0.0102],[40,0.0097],[42,0.0064],[43,0.0079],[44,0.0075],[45,0.0069],[46,0.0017],[47,0.0036],[48,0.0006],[50,-0.0035],[51,-0.0073],[52,-0.0095],[53,-0.0144],[54,-0.0128],[55,-0.013],[56,-0.0163],[57,-0.0176],[58,-0.0182],[59,-0.0171],[60,-0.0167],[61,-0.0167],[62,-0.0116],[63,-0.0114],[64,-0.0078],[65,-0.0078],[66,-0.0072],[67,-0.0053],[68,-0.0026],[69,-0.0007],[70,0.0017],[71,0.0035],[72,0.007],[73,0.0075],[74,0.0076],[75,0.0088],[76,0.0095],[77,0.0112],[78,0.011],[79,0.0083],[80,0.009],[81,0.0052],[82,0.0055],[83,0.0079],[84,0.006],[85,0.006],[86,0.0084],[87,0.0103],[88,0.0125],[89,0.0138],[90,0.0129],[91,0.0134],[92,0.0152],[93,0.0159],[94,0.0172],[95,0.0164],[96,0.0166],[98,0.0185],[99,0.0208],[100,0.0217],[101,0.0196],[102,0.0192],[103,0.0218],[104,0.0201],[105,0.0199],[106,0.0164],[107,0.0184],[108,0.0191],[109,0.0202],[110,0.0178],[111,0.0157],[112,0.0131],[113,0.0084],[114,0.0039],[115,0.0016],[116,-0.0025],[117,-0.0141],[118,-0.0318],[119,-0.0439],[120,-0.0492],[121,-0.0509],[122,-0.0513],[123,-0.0538],[125,-0.0547],[126,-0.0535],[127,-0.0528],[128,-0.0546],[129,-0.0531],[130,-0.0478],[131,-0.0478],[132,-0.0485],[133,-0.0469],[134,-0.0583],[135,-0.0801],[136,-0.0976],[137,-0.1114],[138,-0.095],[139,-0.0774],[140,-0.0623],[141,-0.0597],[142,-0.0539],[143,-0.0533],[144,-0.0541],[145,-0.0554],[146,-0.0564],[147,-0.0566],[148,-0.0547],[149,-0.057],[150,-0.0568],[151,-0.0542],[152,-0.054],[153,-0.0513],[154,-0.0494],[155,-0.0514],[157,-0.0492],[158,-0.0469],[159,-0.0414],[160,-0.0407],[161,-0.0385],[162,-0.0354],[163,-0.037],[164,-0.0353],[165,-0.0343],[166,-0.0276],[167,-0.0242],[168,-0.0248],[169,-0.0242],[170,-0.0195],[171,-0.0142],[172,-0.0153],[173,-0.0196],[174,-0.0227],[175,-0.0269],[176,-0.0295],[177,-0.029],[178,-0.027],[180,-0.0256],[181,-0.0224],[182,-0.0245],[183,-0.0257],[184,-0.0286],[185,-0.029],[186,-0.0288],[187,-0.0277],[188,-0.0276],[189,-0.0314],[190,-0.0329],[191,-0.0371],[192,-0.0603],[193,-0.0779],[194,-0.0827],[195,-0.0819],[196,-0.1072],[197,-0.1478],[198,-0.1793],[199,-0.1893],[200,-0.19],[201,-0.1822],[202,-0.1563],[203,-0.1403],[204,-0.138],[205,-0.1305],[206,-0.1244],[207,-0.1142],[208,-0.1015],[209,-0.095],[210,-0.0862],[211,-0.0772],[212,-0.0781],[213,-0.0771],[214,-0.0752],[215,-0.0726],[216,-0.0686],[217,-0.0643],[219,-0.0641],[220,-0.06],[221,-0.059],[222,-0.0587],[223,-0.0579],[224,-0.0584],[225,-0.0578],[227,-0.057],[228,-0.0593],[229,-0.0618],[230,-0.0598],[231,-0.0583],[232,-0.0575],[233,-0.0725],[234,-0.081],[235,-0.0823],[236,-0.0649],[237,-0.0547],[238,-0.0432],[239,-0.0393],[240,-0.0362],[241,-0.0353],[242,-0.0355],[243,-0.0377],[244,-0.0288],[245,-0.0161],[246,-0.0037],[247,0.0067],[248,0.0114],[249,0.018],[250,0.0162],[251,0.0196],[252,0.0215],[253,0.0254],[254,0.0281],[256,0.0306],[257,0.02],[258,0.0228],[259,0.0245],[260,0.0471],[261,0.0708],[262,0.0834],[263,0.0973],[264,0.0972],[265,0.094],[266,0.0925],[267,0.0867],[268,0.0766],[269,0.0523],[270,0.0478],[271,0.0388],[272,0.0246],[273,0.016],[274,0.0067],[275,0.0057],[276,0.0055],[277,0.0097],[278,0.0124],[279,0.0113],[281,0.0112],[282,0.0105],[283,0.011],[284,0.011],[285,0.0097],[286,0.0087],[287,0.0112],[288,0.0151],[289,0.0193],[290,0.0189],[291,0.0177],[292,0.0185],[293,0.0178],[294,0.02],[295,0.0256],[296,0.0277],[297,0.0277],[298,0.0214],[299,0.0163],[300,0.0062],[301,0.0056],[302,0.0015],[303,0.0004],[304,-0.0026],[305,-0.003],[306,-0.0023],[307,-0.0027],[308,-0.0034],[309,-0.0024],[310,-0.0019],[311,0.0015],[312,0.0019],[313,0.0029],[314,0.0028],[315,0.0002],[317,-0.0006],[318,0.0011],[319,0.0011],[320,0.0025],[321,0.0032],[323,0.0031],[324,0.0061],[325,0.007],[327,0.0069],[329,0.0088],[330,0.009],[331,0.0102],[332,0.011],[333,0.0109],[334,0.0111],[335,0.0121],[336,0.0114],[337,0.0111],[338,0.0131],[339,0.012],[340,0.0119],[341,0.0122],[342,0.0119],[343,0.0114],[344,0.0116],[345,0.0112],[346,0.012],[347,0.0114],[348,0.0101],[349,0.0107],[350,0.0118],[351,0.0139],[352,0.0143],[353,0.0156],[354,0.0158],[355,0.0168],[356,0.017],[357,0.0163],[358,0.0179],[359,0.0167],[360,0.0167]]
       }
       CTRL_L_eye.X {
        type=MovieSceneScriptingFloatChannel
        default=0
        thinned_from=492
        keys [[-131,0.0454],[-130,0.0394],[-129,0.0373],[-128,0.037],[-127,0.0252],[-126,0.0163],[-125,0.008],[-124,0.0028],[-123,-0.0017],[-122,-0.0029],[-121,0.0008],[-120,-0.0044],[-119,0.0005],[-118,0.0101],[-117,0.0109],[-116,0.0158],[-115,0.0198],[-114,0.0282],[-113,0.0333],[-112,0.0362],[-111,0.0397],[-110,0.0442],[-109,0.0448],[-108,0.0427],[-107,0.0415],[-106,0.0409],[-105,0.0316],[-104,0.029],[-103,0.0241],[-102,0.026],[-101,0.0242],[-100,0.0189],[-99,0.0165],[-97,0.0097],[-96,0.0076],[-95,0.0068],[-94,0.0027],[-93,0.0025],[-91,-0.0019],[-90,-0.0048],[-89,-0.0073],[-88,-0.0093],[-86,-0.0191],[-85,-0.0182],[-84,-0.0127],[-83,-0.0109],[-82,-0.0055],[-81,0.0006],[-80,0.0044],[-79,0.005],[-78,0.0101],[-77,0.0198],[-76,0.0193],[-75,0.0222],[-74,0.0209],[-73,0.018],[-72,0.0149],[-71,0.0134],[-70,0.002],[-69,-0.0002],[-68,-0.0004],[-67,-0.001],[-66,0.0028],[-65,0.0108],[-64,0.0103],[-63,0.0137],[-61,0.012],[-60,0.012],[-59,0.0148],[-58,0.0169],[-57,0.0163],[-56,0.0142],[-55,0.0124],[-54,0.0109],[-53,0.0081],[-52,0.0067],[-51,0.0043],[-50,-0.0048],[-49,-0.0059],[-48,-0.0058],[-47,-0.002],[-46,-0.0026],[-45,-0.0042],[-43,-0.0049],[-42,-0.0009],[-41,0.0004],[-40,-0.0016],[-38,0.0003],[-37,0.0088],[-36,0.0096],[-35,0.0113],[-34,0.0179],[-33,0.0217],[-32,0.0249],[-31,0.0258],[-29,0.0289],[-28,0.0311],[-27,0.0309],[-26,0.0287],[-25,0.0257],[-24,0.0251],[-23,0.0224],[-22,0.0689],[-21,0.0705],[-20,0.0505],[-19,0.0266],[-18,0.0214],[-17,0.0221],[-16,0.0201],[-15,0.0159],[-14,0.0103],[-13,0.0112],[-12,0.0107],[-11,0.0082],[-10,0.0085],[-9,0.0036],[-8,0.0039],[-6,0.0075],[-5,-0.0005],[-4,-0.0002],[-3,-0.0007],[-2,0.0004],[-1,0.0048],[0,0.0061],[2,0.0083],[3,0.0117],[5,0.0145],[6,0.0152],[7,0.0142],[8,0.0139],[10,0.0121],[11,0.0116],[12,0.0119],[13,0.0136],[14,0.0128],[15,0.009],[16,0.0069],[17,0.0094],[18,0.0099],[19,0.0098],[20,0.0092],[21,0.0056],[22,0.0064],[24,0.0074],[25,0.0059],[26,0.0076],[27,0.0099],[28,0.0095],[29,0.0084],[30,0.0042],[31,0.0023],[32,0.0053],[33,0.0031],[34,-0.001],[35,-0.0007],[36,0.002],[37,0.002],[38,0.0002],[39,0],[40,-0.0019],[41,-0.0027],[42,-0.0064],[43,-0.0107],[44,-0.0135],[45,-0.0134],[46,-0.0211],[47,-0.0186],[48,-0.0209],[49,-0.0186],[50,-0.0195],[51,-0.0194],[52,-0.0258],[53,-0.0319],[54,-0.0331],[55,-0.0307],[56,-0.0242],[57,-0.0135],[58,-0.0033],[59,-0.0051],[60,-0.0063],[61,-0.0121],[62,-0.0131],[63,-0.0129],[64,-0.0094],[65,-0.0094],[66,-0.0121],[67,-0.0097],[68,-0.0097],[69,-0.0114],[70,-0.0099],[71,-0.0069],[72,-0.006],[74,0.0016],[75,-0.0032],[76,-0.0003],[77,-0.0016],[78,-0.0016],[79,-0.0008],[80,0.004],[81,-0.0017],[82,0.0038],[83,0.0026],[84,0],[86,-0.0023],[87,-0.0028],[88,-0.0038],[89,-0.0088],[90,-0.0089],[91,-0.017],[92,-0.0224],[93,-0.024],[94,-0.0284],[95,-0.0302],[96,-0.0291],[97,-0.0286],[98,-0.0265],[99,-0.0291],[100,-0.0354],[101,-0.0374],[102,-0.0373],[103,-0.0413],[104,-0.0458],[105,-0.0461],[106,-0.0535],[107,-0.0587],[108,-0.0668],[109,-0.0742],[110,-0.0781],[111,-0.0847],[112,-0.0894],[113,-0.1],[114,-0.101],[115,-0.1111],[116,-0.1204],[117,-0.1276],[118,-0.1389],[119,-0.1474],[120,-0.1569],[121,-0.1618],[122,-0.1623],[123,-0.1697],[124,-0.1739],[125,-0.1733],[126,-0.1762],[127,-0.1772],[128,-0.1762],[129,-0.1743],[130,-0.1763],[131,-0.1754],[132,-0.1736],[133,-0.1786],[134,-0.1905],[135,-0.1988],[136,-0.2042],[137,-0.2249],[138,-0.2388],[139,-0.2272],[140,-0.2314],[141,-0.2305],[142,-0.2334],[144,-0.2187],[145,-0.2117],[146,-0.2079],[147,-0.2103],[148,-0.2086],[149,-0.2127],[151,-0.2197],[152,-0.2172],[154,-0.2176],[155,-0.208],[157,-0.2018],[158,-0.1957],[159,-0.1924],[160,-0.1851],[161,-0.1816],[162,-0.1751],[163,-0.1761],[164,-0.1758],[165,-0.1735],[166,-0.1674],[167,-0.1636],[168,-0.156],[169,-0.1545],[170,-0.1507],[172,-0.1412],[173,-0.1326],[174,-0.1227],[175,-0.1151],[176,-0.1122],[177,-0.1111],[178,-0.1063],[179,-0.1005],[180,-0.1006],[181,-0.0981],[182,-0.0942],[183,-0.0864],[184,-0.0801],[185,-0.0746],[186,-0.0623],[187,-0.0395],[188,-0.0353],[189,-0.0083],[190,0.027],[191,0.0593],[192,-0.0114],[193,-0.1183],[194,-0.2796],[195,-0.4059],[196,-0.6238],[197,-0.6916],[198,-0.6221],[199,-0.5062],[200,-0.4753],[201,-0.425],[202,-0.393],[203,-0.3163],[204,-0.2555],[205,-0.197],[206,-0.1587],[207,-0.1265],[208,-0.0914],[209,-0.0691],[210,-0.0608],[211,-0.0451],[212,-0.036],[213,-0.0232],[214,-0.0136],[215,-0.0123],[216,-0.0057],[217,0.0021],[218,0.0067],[219,0.0088],[221,0.0151],[222,0.018],[223,0.0139],[224,0.0153],[225,0.0104],[226,0.004],[227,-0.006],[228,-0.0196],[229,-0.0281],[230,-0.0397],[231,-0.0625],[232,-0.0838],[233,-0.1162],[234,-0.0977],[235,0.0554],[236,0.1467],[237,0.1289],[238,0.1285],[239,0.2624],[240,0.4735],[241,0.7439],[242,0.7949],[243,0.7897],[244,0.6941],[245,0.5877],[246,0.4852],[247,0.456],[248,0.4234],[250,0.3252],[251,0.2851],[252,0.2376],[253,0.2068],[254,0.1723],[255,0.1464],[256,0.1258],[257,0.1011],[258,0.0807],[259,0.0785],[260,0.059],[261,0.051],[262,0.0382],[263,-0.0096],[265,-0.0084],[266,-0.0025],[267,0.034],[268,0.0222],[269,0.0399],[270,0.0458],[271,0.0253],[272,0.0175],[273,0.0209],[274,0.0272],[275,0.0369],[276,0.0533],[277,0.0666],[278,0.0673],[279,0.067],[280,0.067],[281,0.0677],[282,0.0724],[283,0.0693],[284,0.0686],[285,0.0658],[286,0.0691],[287,0.069],[288,0.0681],[289,0.0649],[290,0.0639],[291,0.0717],[292,0.1441],[293,0.1902],[294,0.1884],[295,0.1972],[296,0.1702],[297,0.1637],[298,0.119],[299,0.0935],[300,0.0625],[301,0.0631],[302,0.0662],[303,0.0653],[304,0.0608],[306,0.0461],[307,0.0415],[308,0.0388],[309,0.0372],[310,0.0367],[311,0.0359],[312,0.0347],[313,0.035],[314,0.038],[315,0.0432],[316,0.0498],[317,0.0558],[318,0.0558],[319,0.057],[320,0.0561],[321,0.0515],[322,0.0571],[323,0.0568],[324,0.052],[325,0.05],[326,0.0505],[328,0.0484],[329,0.0505],[330,0.0519],[331,0.0506],[332,0.0507],[333,0.0535],[334,0.0571],[335,0.0579],[336,0.0576],[337,0.0474],[338,0.0454],[339,0.0439],[340,0.044],[341,0.0457],[342,0.0471],[343,0.0434],[345,0.047],[346,0.0489],[347,0.0476],[348,0.0501],[349,0.0651],[350,0.0669],[351,0.0645],[352,0.0614],[353,0.0607],[354,0.0602],[355,0.0593],[356,0.059],[357,0.0572],[358,0.0536],[359,0.0556],[360,0.0556]]
       }
       CTRL_L_eye.Y {
        type=MovieSceneScriptingFloatChannel
        default=0
        thinned_from=492
        keys [[-131,0.4323],[-130,0.3986],[-129,0.4048],[-128,0.4014],[-127,0.3808],[-126,0.3593],[-125,0.3499],[-124,0.3355],[-122,0.3374],[-121,0.3321],[-120,0.3254],[-117,0.3248],[-116,0.3223],[-115,0.3182],[-114,0.3073],[-113,0.2955],[-112,0.2786],[-111,0.2667],[-110,0.2649],[-109,0.2754],[-108,0.2826],[-107,0.2856],[-106,0.2861],[-105,0.2889],[-104,0.2774],[-103,0.2645],[-102,0.2565],[-101,0.2459],[-100,0.2337],[-99,0.2228],[-98,0.2096],[-97,0.1986],[-96,0.1908],[-95,0.1882],[-94,0.1777],[-93,0.177],[-92,0.1776],[-91,0.1847],[-90,0.1898],[-88,0.2154],[-87,0.2318],[-86,0.245],[-84,0.2683],[-83,0.2747],[-82,0.2834],[-80,0.2865],[-79,0.2873],[-78,0.2947],[-77,0.2993],[-76,0.2944],[-75,0.3016],[-74,0.2868],[-73,0.2753],[-72,0.2659],[-71,0.2585],[-70,0.2646],[-69,0.262],[-68,0.263],[-67,0.2662],[-66,0.2677],[-65,0.2721],[-64,0.2721],[-63,0.2801],[-62,0.2791],[-60,0.2802],[-59,0.2783],[-58,0.2838],[-57,0.2836],[-56,0.2826],[-55,0.2838],[-54,0.2814],[-53,0.2803],[-52,0.2819],[-51,0.2845],[-50,0.2911],[-49,0.2918],[-47,0.3205],[-46,0.3227],[-45,0.3283],[-44,0.3295],[-43,0.3345],[-42,0.3421],[-41,0.3444],[-40,0.3377],[-39,0.3272],[-38,0.3238],[-37,0.3154],[-36,0.3092],[-35,0.3052],[-34,0.3074],[-33,0.3043],[-32,0.3026],[-31,0.3036],[-30,0.3056],[-28,0.3069],[-27,0.3084],[-26,0.3079],[-25,0.3003],[-24,0.2959],[-23,0.1797],[-22,0.1231],[-21,0.2244],[-20,0.2526],[-19,0.2528],[-18,0.2582],[-17,0.2694],[-16,0.28],[-15,0.2814],[-14,0.2779],[-13,0.2782],[-12,0.2805],[-11,0.2733],[-10,0.2789],[-9,0.2837],[-8,0.2812],[-7,0.2863],[-6,0.2879],[-5,0.304],[-4,0.3062],[-3,0.3109],[-2,0.3132],[-1,0.3131],[0,0.3057],[1,0.3064],[2,0.3012],[3,0.304],[4,0.3039],[6,0.3109],[7,0.3138],[9,0.3135],[10,0.318],[11,0.3212],[13,0.3235],[14,0.3283],[15,0.3295],[16,0.3287],[17,0.329],[18,0.3332],[19,0.3345],[20,0.3378],[21,0.3327],[22,0.338],[24,0.3403],[25,0.3393],[26,0.344],[27,0.3456],[29,0.3507],[30,0.3567],[31,0.3661],[32,0.3698],[33,0.3709],[34,0.373],[35,0.3742],[36,0.382],[37,0.3759],[38,0.3709],[39,0.3702],[40,0.3753],[41,0.377],[42,0.3732],[43,0.3768],[44,0.3796],[45,0.379],[46,0.3744],[47,0.3675],[48,0.3513],[49,0.3393],[50,0.3215],[51,0.2853],[52,0.2616],[53,0.2012],[54,0.1939],[55,0.1683],[56,0.1421],[57,0.128],[58,0.1265],[59,0.1284],[60,0.1288],[61,0.126],[62,0.1241],[63,0.1233],[64,0.1105],[65,0.1105],[66,0.1005],[67,0.0993],[68,0.1029],[69,0.1033],[70,0.1159],[71,0.1242],[72,0.1341],[73,0.1622],[74,0.1781],[75,0.1886],[76,0.2064],[77,0.225],[78,0.2256],[79,0.2353],[80,0.2479],[81,0.2558],[82,0.2577],[83,0.263],[84,0.2696],[85,0.2751],[86,0.2753],[87,0.2778],[88,0.2858],[89,0.2903],[90,0.2859],[91,0.2869],[92,0.2802],[95,0.28],[96,0.2733],[97,0.2723],[98,0.2683],[99,0.2634],[100,0.2652],[101,0.2694],[102,0.2707],[103,0.2728],[104,0.2793],[105,0.2797],[106,0.2813],[107,0.2846],[108,0.2902],[109,0.2946],[110,0.3046],[111,0.3049],[114,0.3139],[115,0.3287],[116,0.3316],[117,0.3334],[118,0.3377],[119,0.3453],[120,0.3507],[121,0.365],[122,0.3753],[123,0.3956],[124,0.3952],[125,0.4023],[126,0.4037],[127,0.4021],[128,0.4028],[129,0.3999],[130,0.399],[131,0.4014],[132,0.3929],[133,0.3775],[134,0.3743],[135,0.3847],[136,0.4257],[137,0.4986],[138,0.6004],[139,0.6795],[140,0.7391],[141,0.7603],[142,0.7864],[143,0.7819],[145,0.7667],[146,0.769],[147,0.7604],[148,0.7663],[149,0.7604],[150,0.7634],[152,0.7565],[154,0.7535],[155,0.746],[156,0.7483],[158,0.7425],[159,0.7321],[160,0.7452],[161,0.7501],[162,0.7498],[163,0.7523],[164,0.7651],[165,0.7633],[166,0.7699],[167,0.7856],[168,0.8051],[170,0.7992],[172,0.7528],[173,0.7212],[174,0.694],[175,0.6687],[176,0.6512],[177,0.6455],[178,0.6296],[179,0.6275],[180,0.621],[181,0.6182],[183,0.616],[184,0.6116],[185,0.6032],[186,0.5895],[187,0.5526],[188,0.546],[189,0.5188],[190,0.4773],[191,0.4482],[192,0.4478],[193,0.4801],[194,0.6333],[195,0.6625],[196,0.6677],[197,0.6227],[199,0.4125],[200,0.3886],[201,0.34],[202,0.222],[203,0.2733],[204,0.2491],[205,0.2068],[206,0.1778],[207,0.1596],[208,0.1439],[209,0.132],[210,0.167],[211,0.1799],[212,0.1724],[213,0.1531],[214,0.1366],[215,0.126],[216,0.1188],[217,0.1131],[218,0.1151],[219,0.1159],[220,0.115],[221,0.1218],[222,0.1254],[224,0.1298],[225,0.1337],[226,0.1417],[227,0.1513],[228,0.1673],[229,0.1883],[230,0.2128],[231,0.2521],[232,0.2889],[233,0.354],[234,0.3739],[235,0.3339],[236,0.3121],[237,0.2172],[238,0.1057],[239,0.2004],[240,0.1694],[241,0.0538],[242,0.0311],[243,0.0169],[244,0.0555],[245,0.1047],[246,0.1644],[247,0.1685],[248,0.1805],[249,0.2104],[250,0.24],[251,0.2457],[252,0.2703],[253,0.3018],[254,0.3161],[255,0.3342],[256,0.342],[257,0.2681],[258,-0.0051],[259,-0.0419],[260,-0.4233],[261,-0.6592],[262,-0.7545],[263,-0.776],[264,-0.7743],[265,-0.7628],[266,-0.7532],[267,-0.4432],[268,-0.1928],[269,0.2882],[270,0.3805],[271,0.5936],[272,0.5561],[273,0.5125],[274,0.4338],[275,0.3958],[276,0.3982],[277,0.3977],[278,0.3874],[279,0.3851],[280,0.3818],[281,0.3842],[282,0.3849],[283,0.3876],[284,0.3876],[285,0.3801],[286,0.3874],[289,0.3892],[290,0.3886],[291,0.3777],[292,0.3503],[293,0.3333],[294,0.3439],[295,0.3607],[296,0.3601],[297,0.3568],[298,0.2858],[299,0.2364],[300,0.1579],[301,0.1531],[302,0.115],[303,0.1086],[304,0.0895],[305,0.0783],[306,0.0652],[307,0.0566],[308,0.0505],[309,0.0468],[310,0.0456],[311,0.0433],[312,0.0425],[313,0.0428],[314,0.0454],[315,0.0445],[316,0.0531],[317,0.0553],[318,0.062],[319,0.0636],[320,0.0668],[321,0.0604],[323,0.0644],[324,0.0598],[325,0.0603],[327,0.0689],[328,0.066],[329,0.0713],[330,0.0771],[331,0.0795],[332,0.0792],[333,0.0897],[334,0.0949],[335,0.0991],[336,0.1016],[337,0.101],[338,0.105],[339,0.1035],[340,0.1082],[341,0.1077],[342,0.1094],[343,0.1046],[344,0.1079],[345,0.1044],[346,0.1072],[347,0.1067],[348,0.109],[349,0.1158],[350,0.1186],[351,0.1204],[352,0.1227],[353,0.1262],[354,0.1303],[355,0.1324],[356,0.1322],[357,0.1342],[358,0.1391],[359,0.1434],[360,0.1434]]
       }
       CTRL_L_eye_blink {
        type=MovieSceneScriptingFloatChannel
        default=0
        thinned_from=492
        keys [[-131,-0.476],[-130,-0.5154],[-129,-0.5307],[-128,-0.5539],[-127,-0.5738],[-126,-0.5777],[-125,-0.5879],[-124,-0.5835],[-123,-0.5766],[-122,-0.5752],[-121,-0.5811],[-120,-0.5356],[-119,-0.5135],[-118,-0.5035],[-117,-0.5025],[-116,-0.4951],[-115,-0.4894],[-114,-0.492],[-113,-0.4975],[-112,-0.4947],[-111,-0.4935],[-110,-0.4986],[-109,-0.4907],[-108,-0.4839],[-106,-0.4832],[-105,-0.469],[-104,-0.4736],[-103,-0.4845],[-102,-0.4909],[-101,-0.5047],[-100,-0.5091],[-99,-0.4938],[-98,-0.4848],[-97,-0.4804],[-95,-0.4789],[-94,-0.4612],[-93,-0.4608],[-92,-0.462],[-90,-0.4434],[-89,-0.4266],[-88,-0.4215],[-87,-0.4194],[-85,-0.3701],[-84,-0.3607],[-83,-0.361],[-82,-0.3638],[-81,-0.3591],[-80,-0.3504],[-79,-0.3506],[-78,-0.3518],[-77,-0.3383],[-76,-0.3191],[-74,-0.319],[-73,-0.3179],[-72,-0.328],[-71,-0.3227],[-70,-0.3104],[-69,-0.3143],[-68,-0.3213],[-67,-0.322],[-66,-0.3201],[-65,-0.3172],[-64,-0.3236],[-63,-0.318],[-62,-0.3136],[-61,-0.3175],[-59,-0.3097],[-58,-0.3157],[-57,-0.3152],[-55,-0.3103],[-54,-0.3155],[-53,-0.314],[-52,-0.3109],[-51,-0.3033],[-50,-0.3061],[-49,-0.3064],[-48,-0.3119],[-47,-0.3099],[-45,-0.3105],[-44,-0.3144],[-43,-0.3299],[-42,-0.3428],[-41,-0.3424],[-39,-0.338],[-38,-0.337],[-37,-0.3425],[-36,-0.3313],[-35,-0.3237],[-34,-0.3192],[-33,-0.306],[-32,-0.3009],[-31,-0.3004],[-30,-0.2967],[-29,-0.2983],[-28,-0.2971],[-27,-0.2822],[-26,-0.2778],[-25,-0.2707],[-24,-0.2623],[-23,0.0789],[-22,0.7483],[-21,0.6325],[-20,0.3172],[-19,0.0607],[-18,-0.0662],[-17,-0.1803],[-16,-0.2798],[-15,-0.3374],[-14,-0.3725],[-13,-0.3794],[-12,-0.3839],[-11,-0.3811],[-10,-0.3739],[-9,-0.3847],[-8,-0.3853],[-7,-0.3777],[-6,-0.3824],[-5,-0.3902],[-4,-0.4006],[-3,-0.4064],[-2,-0.4041],[-1,-0.3992],[0,-0.3837],[1,-0.3742],[2,-0.3611],[3,-0.3492],[4,-0.3354],[5,-0.3191],[6,-0.2763],[7,-0.2738],[8,-0.2743],[9,-0.2756],[11,-0.2799],[12,-0.2724],[13,-0.263],[14,-0.2571],[16,-0.2609],[17,-0.2656],[18,-0.2636],[19,-0.2532],[20,-0.2569],[21,-0.257],[22,-0.2543],[23,-0.2559],[24,-0.2618],[25,-0.2651],[26,-0.266],[27,-0.2682],[28,-0.2648],[29,-0.2507],[30,-0.2435],[31,-0.2532],[32,-0.2616],[33,-0.2568],[34,-0.2493],[35,-0.2493],[36,-0.2487],[37,-0.244],[38,-0.2433],[39,-0.2438],[40,-0.2479],[42,-0.2527],[43,-0.2559],[44,-0.2632],[45,-0.2659],[46,-0.2751],[47,-0.2781],[48,-0.2722],[49,-0.276],[50,-0.2677],[51,-0.2606],[52,-0.2612],[53,-0.2137],[54,-0.2052],[55,-0.1877],[56,-0.1803],[57,-0.1823],[58,-0.1934],[59,-0.1896],[60,-0.1875],[61,-0.1753],[62,-0.1616],[63,-0.1612],[64,-0.1538],[65,-0.1538],[66,-0.1399],[67,-0.1434],[68,-0.1662],[69,-0.1646],[70,-0.1499],[71,-0.1416],[72,-0.1321],[73,-0.1261],[74,-0.1214],[75,-0.111],[76,-0.1066],[77,-0.1104],[78,-0.1103],[79,-0.1097],[80,-0.118],[81,-0.1185],[82,-0.1233],[83,-0.1372],[84,-0.1356],[85,-0.1388],[86,-0.1432],[87,-0.1444],[88,-0.1462],[89,-0.1453],[90,-0.1516],[91,-0.1591],[92,-0.166],[93,-0.1682],[94,-0.173],[95,-0.1719],[96,-0.1725],[97,-0.1768],[98,-0.1796],[99,-0.1765],[100,-0.1726],[101,-0.1721],[103,-0.1726],[104,-0.1769],[105,-0.1772],[106,-0.1836],[107,-0.1775],[108,-0.1721],[109,-0.1791],[110,-0.1715],[111,-0.1647],[112,-0.1725],[113,-0.1787],[114,-0.1879],[115,-0.1831],[116,-0.1814],[117,-0.1997],[118,-0.2045],[119,-0.1983],[120,-0.1874],[121,-0.1799],[122,-0.1792],[123,-0.189],[124,-0.2039],[125,-0.2074],[126,-0.208],[127,-0.2104],[128,-0.2143],[129,-0.2203],[130,-0.2291],[131,-0.239],[132,-0.2511],[133,-0.2577],[134,-0.2576],[135,-0.2685],[136,-0.3012],[137,-0.3364],[138,-0.363],[139,-0.3975],[140,-0.4301],[141,-0.4363],[142,-0.4628],[143,-0.4718],[144,-0.488],[145,-0.4978],[146,-0.4975],[147,-0.5002],[148,-0.4918],[149,-0.4966],[150,-0.4959],[151,-0.4875],[153,-0.4768],[155,-0.4833],[156,-0.4908],[157,-0.4971],[158,-0.4936],[159,-0.4825],[161,-0.5018],[162,-0.4886],[164,-0.5065],[165,-0.493],[166,-0.4777],[167,-0.4717],[168,-0.4732],[169,-0.4656],[170,-0.4601],[171,-0.4609],[172,-0.474],[173,-0.5014],[174,-0.5167],[175,-0.5227],[176,-0.5251],[177,-0.5229],[178,-0.5173],[180,-0.5336],[181,-0.5278],[182,-0.5329],[183,-0.5348],[184,-0.5307],[185,-0.5219],[186,-0.5193],[187,-0.5349],[188,-0.5378],[189,-0.5459],[190,-0.5345],[191,-0.5087],[192,-0.1639],[193,0.4198],[194,0.5632],[195,0.4126],[196,0.0805],[197,-0.1088],[198,-0.2465],[199,-0.3122],[200,-0.3197],[201,-0.2996],[202,-0.2458],[203,-0.263],[204,-0.3266],[205,-0.3403],[206,-0.3481],[207,-0.3472],[209,-0.3489],[210,-0.3283],[211,-0.3335],[212,-0.3509],[213,-0.3797],[214,-0.4012],[215,-0.4103],[216,-0.41],[217,-0.4127],[218,-0.4173],[219,-0.4085],[220,-0.4038],[221,-0.4041],[222,-0.4011],[224,-0.4002],[225,-0.4021],[226,-0.4051],[227,-0.397],[228,-0.3897],[229,-0.3987],[230,-0.4027],[231,-0.3943],[232,-0.3814],[233,-0.3282],[234,-0.2128],[235,0.078],[236,0.5362],[237,0.7102],[238,0.7303],[239,0.6507],[240,0.4425],[241,0.1831],[242,0.1367],[243,-0.0165],[244,-0.0698],[245,-0.139],[246,-0.2134],[248,-0.3385],[249,-0.3943],[250,-0.4139],[251,-0.4241],[252,-0.4055],[253,-0.3928],[254,-0.4188],[255,-0.4374],[256,-0.4164],[257,-0.2628],[258,-0.116],[259,-0.1255],[260,-0.2707],[261,-0.4476],[262,-0.4851],[263,-0.4254],[264,-0.4189],[265,-0.3991],[266,-0.3628],[267,0.2196],[268,0.6807],[269,0.4488],[270,0.3843],[271,-0.0036],[272,-0.2122],[273,-0.3575],[274,-0.4603],[275,-0.4814],[276,-0.4767],[278,-0.4948],[279,-0.489],[280,-0.489],[281,-0.4983],[282,-0.4982],[283,-0.4929],[284,-0.4929],[285,-0.5081],[286,-0.5293],[287,-0.524],[288,-0.5138],[289,-0.5093],[290,-0.5158],[291,-0.4547],[292,-0.0499],[293,0.5423],[294,0.5229],[295,0.2416],[296,0.0317],[297,0.0002],[298,-0.069],[299,-0.1193],[300,-0.1744],[301,-0.1759],[302,-0.1873],[303,-0.1925],[304,-0.2098],[305,-0.2192],[306,-0.2104],[307,-0.1903],[308,-0.1692],[309,-0.1476],[310,-0.1407],[311,-0.1365],[312,-0.1365],[313,-0.136],[315,-0.1509],[316,-0.155],[317,-0.1575],[318,-0.1554],[319,-0.1504],[320,-0.1476],[321,-0.1461],[322,-0.1547],[323,-0.166],[324,-0.1641],[325,-0.1557],[326,-0.1543],[328,-0.1592],[329,-0.1611],[330,-0.1601],[331,-0.161],[332,-0.1592],[333,-0.1585],[334,-0.1619],[335,-0.1619],[336,-0.1601],[337,-0.1571],[338,-0.1571],[339,-0.1622],[340,-0.1685],[341,-0.1713],[342,-0.168],[343,-0.1699],[344,-0.1762],[345,-0.1743],[346,-0.174],[347,-0.1814],[348,-0.1844],[349,-0.1742],[350,-0.1731],[351,-0.1752],[352,-0.1666],[353,-0.1606],[354,-0.1622],[355,-0.1681],[356,-0.1696],[357,-0.1679],[358,-0.167],[359,-0.1639],[360,-0.1639]]
       }
       CTRL_R_eye.X {
        type=MovieSceneScriptingFloatChannel
        default=0
        thinned_from=492
        keys [[-131,0.0121],[-130,0.0063],[-129,0.0049],[-128,0.004],[-127,-0.0066],[-126,-0.0138],[-125,-0.0214],[-124,-0.0269],[-123,-0.0315],[-122,-0.0327],[-121,-0.0279],[-120,-0.0323],[-119,-0.0275],[-118,-0.016],[-117,-0.0152],[-116,-0.0107],[-115,-0.0064],[-114,0.001],[-113,0.0045],[-112,0.006],[-111,0.0085],[-110,0.0122],[-109,0.0123],[-108,0.0107],[-107,0.0094],[-106,0.0089],[-105,0],[-104,-0.0044],[-103,-0.0084],[-102,-0.0083],[-101,-0.0077],[-100,-0.0139],[-99,-0.016],[-98,-0.0196],[-97,-0.0237],[-96,-0.0271],[-95,-0.0282],[-94,-0.0316],[-93,-0.0317],[-92,-0.0333],[-91,-0.0352],[-90,-0.0377],[-88,-0.0413],[-87,-0.0439],[-86,-0.0472],[-85,-0.0456],[-84,-0.0398],[-83,-0.0374],[-82,-0.0313],[-81,-0.0261],[-80,-0.0224],[-79,-0.0215],[-77,-0.0055],[-76,-0.007],[-75,-0.0015],[-74,-0.0047],[-73,-0.0097],[-72,-0.0116],[-71,-0.0139],[-70,-0.0237],[-69,-0.026],[-68,-0.0262],[-67,-0.0259],[-66,-0.0227],[-65,-0.0149],[-64,-0.015],[-63,-0.0103],[-62,-0.0118],[-61,-0.0139],[-60,-0.0142],[-59,-0.0109],[-58,-0.0098],[-57,-0.0101],[-55,-0.0136],[-54,-0.0141],[-53,-0.0173],[-52,-0.0191],[-51,-0.0229],[-50,-0.0317],[-49,-0.0327],[-48,-0.0303],[-47,-0.0268],[-46,-0.0273],[-45,-0.0288],[-44,-0.0291],[-43,-0.0288],[-42,-0.025],[-41,-0.0251],[-40,-0.0285],[-39,-0.0275],[-38,-0.0264],[-37,-0.017],[-36,-0.0177],[-35,-0.0155],[-34,-0.0095],[-33,-0.005],[-32,-0.0016],[-31,-0.0006],[-30,0.0029],[-29,0.0056],[-28,0.0092],[-27,0.0069],[-26,0.005],[-25,0.0034],[-24,0.0031],[-23,0.0019],[-22,0.1088],[-21,0.1056],[-20,0.0228],[-19,-0.0069],[-18,-0.01],[-17,-0.0048],[-16,-0.0035],[-15,-0.0047],[-14,-0.0089],[-13,-0.0082],[-12,-0.0071],[-11,-0.0098],[-10,-0.0095],[-9,-0.0133],[-8,-0.0138],[-6,-0.008],[-5,-0.0164],[-4,-0.015],[-3,-0.0157],[-2,-0.0137],[-1,-0.01],[0,-0.0079],[1,-0.0071],[2,-0.0074],[3,-0.0052],[4,-0.0046],[5,-0.0012],[6,-0.0004],[7,-0.0014],[8,-0.0017],[11,-0.0045],[12,-0.0044],[13,-0.0022],[14,-0.0026],[15,-0.0072],[16,-0.0102],[17,-0.0077],[18,-0.0078],[19,-0.0088],[20,-0.01],[21,-0.0148],[22,-0.0119],[23,-0.0111],[24,-0.0099],[25,-0.0115],[27,-0.0089],[28,-0.0103],[29,-0.0113],[30,-0.0152],[31,-0.0162],[32,-0.0118],[33,-0.013],[34,-0.0181],[35,-0.0176],[36,-0.014],[37,-0.0153],[38,-0.0172],[39,-0.0178],[40,-0.0198],[41,-0.0194],[42,-0.0222],[43,-0.0262],[44,-0.029],[45,-0.0291],[46,-0.0365],[47,-0.0354],[48,-0.0379],[49,-0.0378],[50,-0.0404],[51,-0.0417],[52,-0.051],[53,-0.0599],[54,-0.0621],[55,-0.0617],[56,-0.0552],[57,-0.0461],[58,-0.0365],[59,-0.0387],[60,-0.0401],[61,-0.0449],[62,-0.0474],[63,-0.0473],[64,-0.045],[65,-0.045],[66,-0.0494],[67,-0.0472],[68,-0.0473],[69,-0.0503],[70,-0.0474],[71,-0.0447],[72,-0.0427],[73,-0.0387],[74,-0.0336],[75,-0.0374],[76,-0.0325],[77,-0.0335],[78,-0.0335],[79,-0.0332],[80,-0.0273],[81,-0.035],[82,-0.0273],[83,-0.0285],[84,-0.0312],[85,-0.032],[86,-0.0338],[87,-0.0338],[88,-0.0363],[89,-0.0398],[90,-0.0402],[91,-0.0478],[92,-0.0543],[93,-0.0554],[94,-0.0584],[95,-0.06],[96,-0.0589],[97,-0.0589],[98,-0.0561],[99,-0.0592],[100,-0.0668],[101,-0.0682],[102,-0.0677],[103,-0.0721],[104,-0.0774],[105,-0.0777],[107,-0.0901],[108,-0.0972],[109,-0.1023],[110,-0.1071],[111,-0.1129],[112,-0.1164],[113,-0.1266],[114,-0.129],[115,-0.14],[116,-0.1496],[117,-0.1575],[118,-0.1684],[119,-0.1759],[120,-0.1858],[121,-0.1919],[122,-0.1906],[124,-0.1988],[125,-0.1976],[126,-0.2015],[127,-0.2033],[128,-0.2019],[129,-0.2011],[130,-0.2021],[131,-0.2013],[132,-0.1992],[133,-0.2047],[134,-0.2181],[135,-0.2265],[136,-0.2335],[137,-0.2525],[138,-0.2668],[139,-0.2478],[140,-0.2483],[141,-0.2451],[142,-0.2473],[143,-0.2396],[144,-0.2349],[145,-0.2294],[146,-0.2261],[147,-0.2278],[148,-0.2244],[149,-0.2303],[150,-0.2342],[151,-0.2371],[152,-0.2343],[153,-0.2371],[154,-0.2375],[155,-0.2283],[157,-0.2219],[158,-0.2147],[159,-0.2097],[160,-0.2019],[161,-0.1978],[162,-0.1903],[163,-0.1933],[164,-0.1905],[165,-0.1884],[166,-0.1811],[167,-0.1773],[168,-0.1694],[169,-0.1701],[170,-0.1662],[171,-0.1599],[172,-0.155],[173,-0.1473],[174,-0.138],[175,-0.1304],[177,-0.1277],[178,-0.1221],[179,-0.116],[180,-0.115],[181,-0.1128],[182,-0.1083],[184,-0.0955],[185,-0.0895],[186,-0.0776],[187,-0.0559],[188,-0.0518],[189,-0.0235],[190,0.011],[191,0.0434],[192,-0.0432],[193,-0.1448],[194,-0.3416],[195,-0.4562],[196,-0.6439],[197,-0.7003],[198,-0.6318],[199,-0.5203],[200,-0.4907],[201,-0.4433],[202,-0.41],[203,-0.3309],[204,-0.2708],[205,-0.2157],[206,-0.1798],[207,-0.1511],[208,-0.117],[209,-0.0948],[210,-0.084],[211,-0.0674],[212,-0.0581],[213,-0.0463],[214,-0.0375],[215,-0.0362],[216,-0.0305],[217,-0.0228],[218,-0.0178],[220,-0.0116],[221,-0.0077],[222,-0.0049],[223,-0.0088],[224,-0.0075],[225,-0.0122],[226,-0.0185],[227,-0.0281],[228,-0.0403],[229,-0.048],[230,-0.0591],[231,-0.0817],[232,-0.1016],[233,-0.1363],[234,-0.121],[235,0.0249],[236,0.146],[237,0.198],[238,0.2198],[239,0.3765],[240,0.564],[241,0.7567],[242,0.7923],[243,0.7673],[244,0.6703],[246,0.4601],[247,0.4324],[248,0.3965],[249,0.344],[250,0.2947],[251,0.2533],[252,0.2077],[253,0.1779],[254,0.1425],[255,0.1183],[256,0.0986],[257,0.0694],[258,0.0429],[259,0.041],[260,0.0351],[261,0.0675],[262,0.0921],[263,0.0402],[264,0.0438],[265,0.0446],[266,0.0382],[267,0.0765],[268,0.0746],[269,0.0308],[270,0.0234],[271,-0.0045],[272,-0.0129],[273,-0.0089],[274,-0.002],[275,0.0087],[276,0.0257],[277,0.0382],[278,0.0391],[279,0.0391],[280,0.0402],[281,0.0408],[282,0.0451],[283,0.0426],[284,0.0419],[285,0.0386],[286,0.0409],[287,0.0411],[288,0.04],[289,0.0354],[290,0.0336],[291,0.0414],[292,0.1119],[293,0.1908],[294,0.1731],[295,0.1527],[296,0.1276],[297,0.1219],[298,0.0805],[299,0.0564],[300,0.0254],[301,0.026],[302,0.0281],[303,0.0268],[304,0.0214],[305,0.0141],[306,0.0059],[307,0.0002],[308,-0.0028],[309,-0.0054],[310,-0.0063],[311,-0.0064],[312,-0.007],[313,-0.0066],[314,-0.0041],[315,0.0011],[316,0.0079],[317,0.0145],[318,0.014],[319,0.0151],[320,0.0141],[321,0.009],[322,0.0155],[323,0.0169],[324,0.0114],[325,0.0091],[326,0.0109],[327,0.0096],[328,0.0104],[329,0.0118],[330,0.0124],[331,0.012],[332,0.0132],[333,0.0175],[334,0.021],[335,0.0218],[336,0.0221],[337,0.0121],[338,0.0109],[339,0.0084],[340,0.0087],[341,0.0111],[342,0.0114],[343,0.0086],[344,0.0109],[346,0.0131],[347,0.0116],[348,0.0145],[349,0.0281],[350,0.0303],[351,0.0268],[352,0.0247],[353,0.0243],[354,0.0242],[355,0.0232],[356,0.0229],[357,0.0211],[358,0.0182],[359,0.0203],[360,0.0203]]
       }
       CTRL_R_eye.Y {
        type=MovieSceneScriptingFloatChannel
        default=0
        thinned_from=492
        keys [[-131,0.431],[-130,0.3968],[-129,0.4029],[-128,0.3995],[-126,0.3584],[-125,0.3493],[-124,0.3346],[-122,0.3364],[-121,0.3311],[-120,0.323],[-119,0.3224],[-118,0.3235],[-117,0.3232],[-116,0.3204],[-115,0.3165],[-114,0.3058],[-113,0.2937],[-112,0.277],[-111,0.2656],[-110,0.2638],[-109,0.274],[-108,0.2817],[-107,0.2842],[-106,0.2846],[-105,0.2877],[-104,0.2757],[-103,0.2626],[-102,0.2546],[-101,0.244],[-100,0.2318],[-99,0.2213],[-98,0.2078],[-97,0.1971],[-96,0.1887],[-95,0.1858],[-94,0.1759],[-93,0.1753],[-92,0.1763],[-91,0.1824],[-90,0.1879],[-89,0.2013],[-88,0.2134],[-87,0.2294],[-86,0.2426],[-84,0.2658],[-83,0.272],[-82,0.2805],[-79,0.2828],[-78,0.29],[-77,0.2962],[-76,0.2925],[-75,0.3003],[-74,0.2864],[-73,0.2752],[-72,0.2659],[-71,0.2591],[-70,0.2642],[-69,0.2632],[-68,0.2635],[-67,0.267],[-66,0.2687],[-65,0.2725],[-64,0.2728],[-63,0.2804],[-62,0.2799],[-60,0.2815],[-59,0.2791],[-58,0.2855],[-57,0.2854],[-56,0.2845],[-55,0.2854],[-54,0.2834],[-53,0.2829],[-51,0.2861],[-50,0.2928],[-49,0.2935],[-48,0.3072],[-47,0.3215],[-46,0.3235],[-45,0.3291],[-44,0.3301],[-43,0.3345],[-42,0.3411],[-41,0.3428],[-40,0.3351],[-39,0.3241],[-38,0.3208],[-37,0.3129],[-36,0.3071],[-35,0.3046],[-34,0.3065],[-33,0.304],[-32,0.3025],[-31,0.3038],[-30,0.3067],[-28,0.3079],[-27,0.3094],[-26,0.3088],[-25,0.3017],[-24,0.2975],[-23,0.181],[-22,0.1214],[-21,0.2219],[-20,0.2511],[-19,0.2517],[-18,0.2572],[-17,0.269],[-16,0.2799],[-15,0.2814],[-14,0.2772],[-13,0.2773],[-12,0.2796],[-11,0.2737],[-9,0.283],[-8,0.281],[-7,0.2868],[-6,0.2881],[-5,0.304],[-4,0.3056],[-3,0.3114],[-2,0.3134],[-1,0.3131],[0,0.3057],[1,0.3059],[2,0.3009],[3,0.3039],[4,0.3032],[5,0.3074],[7,0.3129],[9,0.3118],[10,0.3164],[11,0.32],[12,0.3215],[13,0.3223],[14,0.3266],[15,0.3281],[16,0.3272],[17,0.3277],[18,0.3316],[19,0.3333],[20,0.3368],[21,0.3317],[22,0.337],[24,0.3386],[25,0.3374],[26,0.3421],[28,0.3464],[29,0.3493],[30,0.3551],[31,0.3648],[32,0.3687],[33,0.3693],[34,0.3719],[35,0.3731],[36,0.3809],[37,0.3744],[38,0.3695],[39,0.3686],[40,0.3739],[41,0.3752],[42,0.3716],[43,0.3756],[44,0.3785],[45,0.378],[46,0.3734],[47,0.3672],[48,0.3506],[49,0.339],[50,0.322],[51,0.2853],[52,0.261],[53,0.2008],[54,0.1937],[55,0.1682],[56,0.1416],[57,0.1276],[58,0.1259],[59,0.1275],[60,0.1279],[62,0.1238],[63,0.123],[64,0.1096],[65,0.1096],[66,0.0996],[67,0.099],[68,0.1028],[69,0.1033],[70,0.1159],[71,0.1242],[72,0.1334],[73,0.1622],[74,0.1777],[75,0.1878],[77,0.2259],[78,0.2264],[79,0.2348],[80,0.2471],[81,0.2549],[82,0.2574],[83,0.2629],[84,0.2696],[85,0.2752],[86,0.276],[87,0.2786],[88,0.2867],[89,0.2902],[90,0.2864],[91,0.2877],[92,0.281],[95,0.2803],[96,0.2734],[97,0.2725],[98,0.2687],[99,0.2635],[100,0.2653],[101,0.2698],[103,0.2732],[104,0.28],[105,0.2804],[106,0.2821],[107,0.2849],[109,0.2953],[110,0.3049],[111,0.3056],[112,0.3082],[113,0.312],[114,0.3148],[115,0.3297],[116,0.3326],[117,0.3343],[118,0.338],[119,0.3454],[120,0.3508],[121,0.365],[122,0.3753],[123,0.3956],[124,0.3952],[125,0.4029],[126,0.4045],[127,0.4017],[128,0.4034],[129,0.4008],[130,0.4005],[131,0.4021],[132,0.3938],[133,0.3791],[134,0.3754],[135,0.3851],[136,0.4258],[137,0.4993],[138,0.6026],[139,0.6808],[140,0.7395],[141,0.761],[142,0.7867],[143,0.782],[145,0.7679],[146,0.7686],[147,0.7603],[148,0.767],[149,0.7613],[150,0.7644],[152,0.7566],[154,0.7535],[155,0.7466],[156,0.7492],[158,0.7443],[159,0.7341],[160,0.7465],[161,0.7512],[162,0.7508],[163,0.7532],[164,0.7654],[165,0.764],[166,0.7708],[167,0.7866],[168,0.8061],[169,0.8043],[170,0.8004],[172,0.7533],[173,0.7213],[175,0.6689],[176,0.6519],[177,0.6462],[178,0.6298],[179,0.6276],[180,0.621],[181,0.6174],[183,0.6158],[184,0.612],[185,0.6035],[186,0.5896],[187,0.5526],[188,0.546],[189,0.5188],[190,0.4772],[191,0.4476],[192,0.4462],[193,0.4776],[194,0.6293],[195,0.6585],[196,0.6637],[197,0.6188],[199,0.4098],[200,0.3858],[201,0.3378],[202,0.2194],[203,0.2711],[204,0.2478],[205,0.2043],[206,0.1753],[207,0.158],[208,0.1421],[209,0.1309],[210,0.1666],[211,0.1791],[212,0.1715],[213,0.1522],[214,0.135],[215,0.1253],[216,0.118],[217,0.1122],[218,0.1141],[219,0.1156],[220,0.1149],[221,0.1211],[222,0.1252],[223,0.1276],[224,0.1304],[225,0.1345],[226,0.142],[227,0.1514],[228,0.1666],[229,0.1881],[230,0.2128],[231,0.2521],[232,0.2883],[233,0.3531],[234,0.3729],[235,0.3321],[236,0.3086],[237,0.213],[238,0.1011],[239,0.195],[240,0.1649],[241,0.0508],[242,0.0284],[243,0.014],[244,0.0526],[245,0.1018],[246,0.1619],[247,0.1652],[248,0.1774],[249,0.2075],[250,0.2371],[251,0.2434],[252,0.2675],[253,0.2989],[254,0.3126],[255,0.3311],[256,0.3396],[257,0.2652],[258,-0.0079],[259,-0.0448],[260,-0.4277],[261,-0.6653],[262,-0.7612],[263,-0.7828],[264,-0.7813],[266,-0.7599],[267,-0.4493],[268,-0.1982],[269,0.285],[270,0.3777],[271,0.5907],[272,0.5545],[273,0.5107],[274,0.4317],[275,0.3933],[276,0.3961],[277,0.3957],[278,0.3854],[279,0.3835],[280,0.3802],[281,0.3819],[282,0.3827],[283,0.3855],[284,0.3856],[285,0.3781],[286,0.386],[289,0.3873],[290,0.3867],[291,0.3757],[292,0.3474],[293,0.3289],[294,0.3396],[295,0.3578],[296,0.3573],[297,0.354],[298,0.2839],[299,0.235],[300,0.1566],[301,0.1519],[302,0.1147],[303,0.1086],[304,0.0902],[305,0.0792],[306,0.0662],[307,0.0576],[308,0.0519],[309,0.0481],[310,0.0468],[311,0.0435],[312,0.0426],[313,0.0429],[314,0.0455],[315,0.0458],[316,0.0537],[317,0.0561],[318,0.0623],[319,0.0637],[320,0.0668],[321,0.0604],[322,0.0629],[323,0.0645],[324,0.0598],[325,0.0604],[327,0.0689],[328,0.066],[329,0.0712],[330,0.0758],[331,0.079],[332,0.0791],[333,0.0891],[334,0.0947],[335,0.099],[336,0.1016],[337,0.101],[338,0.1057],[339,0.1037],[340,0.1083],[341,0.1078],[342,0.1087],[343,0.1044],[344,0.1078],[345,0.1044],[346,0.1071],[347,0.1061],[348,0.1087],[349,0.1152],[350,0.1183],[351,0.1204],[352,0.1227],[353,0.1254],[354,0.1294],[355,0.1315],[356,0.1313],[357,0.134],[358,0.139],[359,0.1427],[360,0.1427]]
       }
       CTRL_R_eye_blink {
        type=MovieSceneScriptingFloatChannel
        default=0
        thinned_from=492
        keys [[-131,-0.6387],[-130,-0.6681],[-129,-0.6698],[-128,-0.6849],[-127,-0.7055],[-125,-0.7129],[-123,-0.7014],[-122,-0.7005],[-121,-0.7046],[-120,-0.6694],[-119,-0.6555],[-118,-0.6464],[-117,-0.6455],[-116,-0.6402],[-115,-0.6379],[-114,-0.6381],[-113,-0.6449],[-112,-0.6496],[-111,-0.6427],[-110,-0.6395],[-109,-0.6425],[-108,-0.637],[-107,-0.6222],[-106,-0.6201],[-104,-0.6179],[-103,-0.6129],[-102,-0.6148],[-101,-0.6233],[-100,-0.627],[-99,-0.6039],[-98,-0.5937],[-97,-0.581],[-96,-0.5724],[-95,-0.5701],[-94,-0.5548],[-93,-0.5528],[-92,-0.5431],[-91,-0.5398],[-89,-0.5166],[-88,-0.5205],[-87,-0.5297],[-86,-0.5226],[-85,-0.5066],[-84,-0.4862],[-83,-0.4843],[-81,-0.4863],[-80,-0.4805],[-78,-0.4806],[-77,-0.4843],[-76,-0.471],[-75,-0.4722],[-74,-0.4759],[-73,-0.4732],[-72,-0.4681],[-70,-0.4616],[-69,-0.459],[-68,-0.4623],[-67,-0.4616],[-66,-0.4565],[-65,-0.4551],[-64,-0.4616],[-63,-0.4695],[-62,-0.4699],[-61,-0.4684],[-60,-0.4636],[-59,-0.462],[-58,-0.4756],[-57,-0.4727],[-56,-0.4637],[-55,-0.4563],[-54,-0.4543],[-53,-0.4456],[-52,-0.4402],[-51,-0.4499],[-50,-0.451],[-49,-0.4509],[-48,-0.4602],[-47,-0.4655],[-45,-0.4649],[-44,-0.4684],[-43,-0.4822],[-42,-0.4932],[-41,-0.4949],[-40,-0.4888],[-39,-0.4704],[-38,-0.4648],[-37,-0.4838],[-36,-0.4711],[-35,-0.4604],[-34,-0.4595],[-33,-0.4565],[-32,-0.4544],[-29,-0.4574],[-28,-0.457],[-27,-0.4505],[-26,-0.4543],[-25,-0.4392],[-24,-0.4252],[-23,0.0037],[-22,0.8021],[-21,0.6935],[-20,0.3058],[-19,0.0604],[-18,-0.0656],[-17,-0.2012],[-16,-0.3268],[-15,-0.4085],[-14,-0.454],[-13,-0.4519],[-12,-0.4587],[-11,-0.4585],[-10,-0.4607],[-9,-0.4846],[-8,-0.4803],[-7,-0.4654],[-6,-0.4753],[-5,-0.4815],[-4,-0.4896],[-3,-0.5079],[-1,-0.5115],[0,-0.4974],[1,-0.4733],[2,-0.4434],[4,-0.4339],[5,-0.4239],[6,-0.4034],[7,-0.4083],[8,-0.4061],[10,-0.3892],[12,-0.3895],[13,-0.3912],[14,-0.388],[16,-0.3776],[17,-0.3741],[18,-0.3629],[19,-0.3614],[20,-0.3709],[21,-0.3643],[23,-0.3671],[25,-0.3731],[26,-0.3841],[27,-0.3845],[28,-0.3859],[29,-0.3847],[30,-0.3943],[31,-0.4128],[32,-0.4019],[33,-0.3947],[34,-0.4004],[35,-0.4006],[36,-0.3989],[37,-0.3827],[38,-0.3896],[39,-0.4013],[40,-0.3978],[41,-0.4025],[42,-0.4146],[43,-0.4127],[44,-0.4139],[45,-0.4167],[46,-0.4151],[47,-0.4075],[49,-0.3988],[50,-0.3924],[51,-0.4207],[52,-0.4268],[53,-0.3645],[54,-0.3628],[55,-0.3535],[56,-0.3407],[57,-0.3336],[58,-0.3358],[59,-0.3399],[61,-0.3414],[62,-0.3389],[63,-0.3379],[64,-0.322],[65,-0.322],[66,-0.3073],[67,-0.301],[68,-0.3051],[69,-0.3021],[70,-0.2904],[71,-0.2872],[72,-0.2806],[73,-0.2828],[74,-0.2839],[75,-0.2702],[76,-0.2686],[77,-0.2842],[78,-0.2844],[79,-0.2872],[80,-0.2992],[81,-0.3052],[82,-0.3094],[83,-0.3149],[84,-0.3239],[85,-0.337],[86,-0.3384],[87,-0.3484],[88,-0.3549],[89,-0.3568],[90,-0.3649],[91,-0.3696],[92,-0.3831],[93,-0.3869],[94,-0.3953],[95,-0.3928],[96,-0.3864],[97,-0.383],[98,-0.3849],[99,-0.3835],[100,-0.3769],[101,-0.3826],[102,-0.3858],[103,-0.3818],[104,-0.3827],[105,-0.3828],[106,-0.3903],[107,-0.3862],[108,-0.3748],[109,-0.3719],[110,-0.3744],[111,-0.3753],[113,-0.3844],[114,-0.3902],[115,-0.3936],[116,-0.3788],[117,-0.3726],[118,-0.3673],[119,-0.3418],[120,-0.3195],[121,-0.3148],[122,-0.3169],[123,-0.3292],[124,-0.3428],[125,-0.3449],[126,-0.3448],[127,-0.348],[129,-0.3657],[130,-0.3735],[131,-0.3788],[132,-0.3923],[133,-0.3919],[134,-0.3866],[136,-0.4121],[137,-0.4387],[138,-0.4607],[139,-0.4767],[140,-0.5188],[141,-0.5434],[142,-0.5714],[143,-0.57],[144,-0.5708],[145,-0.5766],[146,-0.5973],[147,-0.5961],[148,-0.5887],[149,-0.6021],[150,-0.6042],[151,-0.5911],[152,-0.5876],[153,-0.5896],[154,-0.5976],[155,-0.5914],[156,-0.5967],[157,-0.6157],[158,-0.6172],[159,-0.598],[160,-0.5932],[161,-0.6104],[162,-0.6026],[163,-0.6073],[164,-0.6158],[165,-0.6029],[167,-0.6147],[168,-0.6164],[169,-0.6114],[170,-0.6092],[171,-0.6178],[172,-0.6366],[173,-0.6504],[174,-0.6598],[175,-0.6753],[177,-0.6767],[178,-0.6796],[179,-0.6851],[180,-0.6855],[181,-0.6844],[183,-0.6918],[185,-0.6943],[186,-0.7044],[187,-0.7113],[188,-0.712],[189,-0.7072],[190,-0.6953],[191,-0.6748],[192,-0.2838],[193,0.2818],[194,0.2892],[195,0.1503],[196,-0.087],[197,-0.2764],[198,-0.4593],[199,-0.5305],[200,-0.5391],[201,-0.4578],[202,-0.4066],[203,-0.4206],[204,-0.4062],[205,-0.3903],[206,-0.3801],[208,-0.3733],[209,-0.3584],[210,-0.3417],[211,-0.3674],[212,-0.3851],[213,-0.404],[214,-0.4203],[215,-0.4347],[217,-0.4329],[218,-0.4385],[219,-0.4394],[220,-0.4442],[221,-0.4474],[222,-0.4471],[224,-0.4491],[225,-0.4524],[226,-0.4513],[228,-0.429],[229,-0.4262],[230,-0.4209],[231,-0.4112],[232,-0.3977],[233,-0.3393],[234,-0.2456],[235,-0.0189],[236,0.4067],[237,0.6776],[238,0.7659],[239,0.5225],[240,0.2077],[241,-0.0153],[242,-0.0501],[243,-0.2506],[244,-0.443],[245,-0.5206],[246,-0.5698],[247,-0.634],[248,-0.6902],[249,-0.7101],[250,-0.6941],[251,-0.6862],[253,-0.6412],[255,-0.6579],[256,-0.6351],[257,-0.4223],[258,-0.1484],[259,-0.1535],[260,-0.2546],[261,-0.3915],[262,-0.3741],[263,-0.2957],[264,-0.2969],[265,-0.2774],[266,-0.2315],[267,0.3021],[268,0.7042],[269,0.3336],[270,0.2429],[271,-0.0896],[272,-0.3624],[273,-0.5081],[274,-0.6169],[275,-0.6451],[276,-0.6506],[277,-0.6654],[278,-0.6861],[279,-0.6895],[280,-0.6878],[281,-0.6951],[282,-0.6995],[283,-0.6965],[284,-0.6963],[285,-0.701],[286,-0.7242],[287,-0.7223],[288,-0.7161],[289,-0.7059],[290,-0.703],[291,-0.6276],[292,-0.1796],[293,0.4814],[294,0.4183],[295,0.0977],[296,-0.0395],[297,-0.0605],[298,-0.1899],[299,-0.2484],[300,-0.2936],[301,-0.2953],[302,-0.3076],[303,-0.3127],[304,-0.329],[305,-0.3355],[306,-0.322],[307,-0.2951],[308,-0.2611],[309,-0.2278],[310,-0.2173],[311,-0.2018],[312,-0.1992],[313,-0.199],[314,-0.2087],[315,-0.2192],[316,-0.2291],[317,-0.2361],[318,-0.242],[319,-0.2429],[320,-0.2422],[321,-0.2304],[322,-0.2385],[323,-0.2583],[324,-0.2583],[325,-0.2523],[326,-0.2546],[327,-0.2607],[329,-0.2607],[330,-0.2655],[331,-0.2681],[332,-0.2649],[333,-0.2682],[334,-0.2784],[335,-0.2814],[336,-0.2821],[337,-0.2754],[338,-0.2814],[339,-0.2976],[340,-0.3009],[341,-0.2967],[342,-0.2938],[343,-0.2956],[344,-0.2994],[345,-0.3005],[346,-0.3035],[347,-0.3079],[348,-0.3088],[349,-0.3124],[350,-0.3211],[351,-0.3305],[352,-0.3287],[353,-0.3257],[354,-0.3307],[355,-0.3365],[356,-0.3374],[357,-0.3364],[358,-0.3369],[359,-0.3339],[360,-0.3339]]
       }
       CTRL_C_eye.X {type=MovieSceneScriptingFloatChannel, default=0.6, const [0,492]}
       CTRL_C_eye.Y {type=MovieSceneScriptingFloatChannel, default=-0.65, const [0,492]}
       CTRL_C_eyesConvergence {type=MovieSceneScriptingFloatChannel, default=0, const [0,492]}
       CTRL_L_eye_cheekRaise {
        type=MovieSceneScriptingFloatChannel
        default=0
        thinned_from=492
        keys [[-131,0.0032],[-129,0.0023],[-128,0.0015],[-126,0.0007],[-125,0.0008],[-124,0.0012],[-122,0.0011],[-121,0.0007],[-120,0.0012],[-117,0.0014],[-115,0.0021],[-109,0.0012],[-108,0.0014],[-107,0.0019],[-105,0.0022],[-104,0.0021],[-103,0.0019],[-99,0.0012],[-94,0.0019],[-93,0.0019],[-91,0.0013],[-90,0.0014],[-86,0.0031],[-83,0.0036],[-78,0.0035],[-73,0.0028],[-66,0.0033],[-63,0.003],[-54,0.0034],[-51,0.0038],[-43,0.0038],[-40,0.0033],[-35,0.0034],[-31,0.0033],[-30,0.0043],[-29,0.0043],[-28,0.0046],[-26,0.0068],[-25,0.0071],[-24,0.0071],[-23,0.0054],[-21,0.0033],[-20,0.0035],[-19,0.0033],[-15,0.0034],[-12,0.0037],[-7,0.0037],[-6,0.0038],[-3,0.0037],[-1,0.0039],[4,0.0037],[5,0.0039],[6,0.0096],[7,0.0126],[8,0.0133],[9,0.0147],[11,0.015],[14,0.0124],[15,0.0121],[17,0.0118],[18,0.0105],[19,0.0086],[20,0.0082],[22,0.0099],[23,0.0099],[25,0.0096],[27,0.0115],[28,0.0123],[29,0.0127],[35,0.0122],[36,0.0117],[38,0.0117],[39,0.0122],[40,0.0123],[41,0.012],[42,0.0115],[43,0.0112],[44,0.0106],[45,0.0104],[47,0.0084],[48,0.0069],[49,0.006],[50,0.0058],[51,0.0061],[52,0.0056],[53,0.0042],[54,0.0048],[55,0.0061],[56,0.0067],[57,0.0068],[58,0.0065],[60,0.0064],[61,0.0058],[62,0.0068],[63,0.007],[64,0.0097],[66,0.0095],[67,0.0091],[68,0.0075],[69,0.0091],[70,0.0094],[71,0.0115],[72,0.0138],[73,0.0181],[74,0.0191],[75,0.021],[77,0.029],[78,0.029],[79,0.0302],[80,0.0327],[81,0.0317],[82,0.0335],[83,0.0342],[84,0.0359],[85,0.035],[86,0.0336],[88,0.0363],[89,0.0383],[90,0.0372],[91,0.0384],[92,0.0361],[93,0.0363],[95,0.0375],[96,0.0362],[97,0.0334],[98,0.0322],[99,0.0336],[100,0.0338],[101,0.0313],[102,0.0306],[103,0.0288],[104,0.0298],[105,0.0299],[106,0.0264],[108,0.0247],[109,0.0273],[110,0.0256],[111,0.0237],[112,0.0226],[113,0.0223],[115,0.0246],[116,0.0267],[117,0.0267],[118,0.0253],[119,0.0231],[120,0.0223],[121,0.0262],[122,0.0311],[123,0.0335],[124,0.0326],[125,0.032],[127,0.0336],[128,0.0336],[129,0.0345],[130,0.0357],[132,0.0361],[133,0.0371],[134,0.0385],[135,0.0361],[136,0.0297],[137,0.0199],[138,0.0206],[139,0.0247],[140,0.032],[141,0.0319],[142,0.0293],[143,0.0285],[144,0.0284],[145,0.0279],[146,0.0291],[148,0.0303],[149,0.0291],[150,0.031],[151,0.0332],[152,0.0349],[153,0.0343],[155,0.0315],[156,0.0309],[157,0.0316],[159,0.0313],[160,0.0299],[161,0.0306],[162,0.0292],[163,0.0285],[164,0.0272],[165,0.0255],[166,0.0224],[167,0.019],[168,0.0165],[169,0.0149],[170,0.0156],[171,0.0164],[172,0.0162],[173,0.0141],[175,0.0073],[176,0.0046],[178,0.0043],[179,0.0035],[180,0.0032],[181,0.0024],[182,0.0023],[184,0.0017],[187,0.0024],[188,0.0024],[191,0.0018],[192,0.0018],[193,0.0025],[194,0.0039],[195,0.004],[196,0.0037],[197,0.003],[198,0.0026],[200,0.0029],[201,0.0034],[202,0.0045],[204,0.0047],[205,0.0035],[208,0.0035],[210,0.0033],[212,0.0033],[214,0.0028],[219,0.0028],[221,0.0032],[228,0.0031],[234,0.0028],[235,0.0024],[236,0.0025],[238,0.0033],[240,0.0021],[243,0.0019],[244,0.0015],[247,0.0013],[248,0.0007],[249,0.0008],[250,0.0004],[251,0.001],[253,0.0013],[255,0.0015],[256,0.0013],[257,0.0004],[258,0],[259,0.0002],[261,0.002],[263,0.0031],[265,0.0038],[271,0.0036],[273,0.0026],[274,0.0017],[275,0.0012],[280,0.0011],[285,0.0014],[286,0.0008],[288,0.0008],[289,0.0012],[290,0.0012],[292,0.0009],[296,0.0018],[299,0.0021],[300,0.0026],[305,0.0027],[308,0.0032],[310,0.003],[311,0.0027],[321,0.0026],[323,0.0021],[324,0.0021],[326,0.0027],[329,0.0023],[333,0.0028],[340,0.0028],[342,0.0024],[345,0.0026],[348,0.0024],[352,0.0026],[353,0.0029],[356,0.0028],[357,0.0027],[358,0.0028],[360,0.0028]]
       }
       CTRL_L_eye_squintInner {
        type=MovieSceneScriptingFloatChannel
        default=0
        thinned_from=492
        keys [[-131,0.0015],[-130,0.001],[-129,0.0001],[-117,0],[-114,0.0004],[-111,0.0002],[-110,0],[-108,0.0002],[-107,0.0005],[-106,0.0005],[-103,0],[-98,0],[-95,0.0002],[-90,0],[-88,0.0004],[-84,0.002],[-82,0.0022],[-81,0.002],[-79,0.0021],[-78,0.0023],[-77,0.0023],[-75,0.0024],[-74,0.0022],[-72,0.0028],[-71,0.0026],[-67,0.0027],[-66,0.0026],[-65,0.0022],[-60,0.0023],[-59,0.0027],[-58,0.0028],[-56,0.0024],[-55,0.002],[-54,0.0019],[-48,0.0023],[-45,0.0027],[-42,0.0022],[-38,0.0026],[-36,0.0024],[-34,0.003],[-33,0.0076],[-32,0.0112],[-31,0.0136],[-30,0.0317],[-29,0.0277],[-28,0.02],[-27,0.0307],[-26,0.0353],[-25,0.023],[-24,0.0185],[-23,0.0069],[-22,0.0018],[-21,0.0018],[-20,0.0036],[-19,0.0078],[-18,0.0088],[-17,0.0049],[-16,0.0024],[-14,0.0023],[-13,0.0027],[-11,0.0028],[-10,0.0031],[-8,0.003],[-7,0.0032],[-4,0.0029],[-2,0.0031],[-1,0.0029],[1,0.0036],[2,0.0044],[3,0.0058],[4,0.0122],[5,0.0281],[6,0.0981],[7,0.1239],[8,0.1273],[9,0.1346],[10,0.1386],[11,0.1341],[12,0.1263],[13,0.1259],[14,0.1386],[16,0.1582],[17,0.1664],[18,0.1703],[19,0.1623],[20,0.1688],[21,0.1803],[22,0.1743],[23,0.1731],[24,0.1699],[25,0.1701],[26,0.1758],[27,0.1801],[29,0.1854],[30,0.1869],[31,0.1806],[32,0.169],[33,0.1719],[34,0.1741],[35,0.1743],[36,0.1767],[37,0.1802],[38,0.18],[40,0.1848],[41,0.1758],[42,0.163],[43,0.1547],[44,0.1496],[45,0.1488],[46,0.1366],[47,0.1083],[48,0.0791],[49,0.0747],[50,0.083],[51,0.0781],[52,0.0747],[53,0.0711],[54,0.0869],[55,0.1033],[56,0.0872],[57,0.0884],[58,0.1078],[59,0.099],[60,0.0952],[61,0.0878],[62,0.0865],[63,0.0871],[64,0.0962],[65,0.0962],[66,0.0863],[67,0.078],[68,0.0774],[69,0.0791],[70,0.0855],[71,0.0933],[72,0.1088],[73,0.1196],[74,0.1327],[75,0.1347],[76,0.145],[77,0.1561],[78,0.1565],[79,0.1629],[80,0.1739],[81,0.174],[82,0.1646],[83,0.1707],[84,0.1731],[85,0.1637],[86,0.157],[87,0.1653],[88,0.1694],[89,0.1686],[90,0.1619],[91,0.1546],[92,0.144],[93,0.146],[94,0.1508],[95,0.1502],[96,0.1455],[97,0.1464],[98,0.1494],[99,0.1479],[100,0.1459],[101,0.1379],[102,0.1343],[103,0.1256],[104,0.1299],[105,0.1302],[106,0.1177],[107,0.1113],[108,0.123],[109,0.1201],[110,0.1132],[111,0.1103],[112,0.1088],[113,0.1087],[114,0.1167],[115,0.1258],[116,0.1146],[117,0.0993],[118,0.0877],[119,0.0746],[120,0.0731],[121,0.1052],[122,0.1366],[123,0.1511],[124,0.1539],[125,0.1623],[126,0.1776],[127,0.1712],[128,0.1692],[129,0.1737],[130,0.1717],[132,0.1629],[133,0.1578],[134,0.1553],[135,0.1279],[136,0.0689],[137,0.0366],[138,0.0817],[139,0.1549],[140,0.1683],[141,0.1573],[142,0.1435],[143,0.1438],[144,0.1336],[145,0.1217],[146,0.1249],[147,0.1297],[149,0.1278],[150,0.1359],[151,0.1468],[152,0.1462],[153,0.1447],[154,0.1472],[155,0.1393],[156,0.1311],[157,0.138],[158,0.1286],[159,0.1219],[160,0.1255],[161,0.1328],[162,0.1242],[163,0.1107],[164,0.109],[165,0.0946],[166,0.0657],[167,0.0632],[168,0.0576],[169,0.0559],[170,0.0645],[171,0.0598],[172,0.0426],[173,0.0325],[175,0.011],[176,0.0071],[177,0.0067],[178,0.0056],[179,0.0029],[180,0.0006],[181,0.0008],[182,0.0006],[183,0.0001],[184,0],[186,0.0004],[188,0.0004],[190,0],[191,0],[192,0.0002],[194,0.0027],[195,0.0029],[196,0.0027],[197,0.0018],[198,0.0019],[199,0.0029],[200,0.0032],[201,0.0025],[202,0.009],[203,0.0101],[204,0.0031],[205,0.0025],[210,0.0022],[212,0.0023],[213,0.0019],[214,0.0017],[216,0.0018],[218,0.0015],[219,0.0016],[220,0.0021],[221,0.0023],[224,0.0018],[226,0.0022],[228,0.0021],[231,0.0016],[232,0.0016],[233,0.0021],[234,0.0022],[235,0.001],[237,0.0021],[238,0.0018],[239,0.0007],[240,0.0006],[241,0.0009],[242,0.001],[243,0.0002],[244,0],[246,0.0001],[253,0],[255,0.0001],[259,0],[261,0.0013],[262,0.0012],[263,0.0027],[265,0.0032],[266,0.0031],[269,0.0019],[271,0.0019],[272,0.0017],[274,0],[285,0.0001],[294,0],[296,0.0007],[297,0.0007],[298,0.0003],[299,0.0005],[300,0.0011],[304,0.001],[307,0.0016],[316,0.0014],[318,0.0007],[320,0.0013],[321,0.0015],[323,0.0007],[325,0.0017],[328,0.0016],[330,0.0016],[332,0.0021],[334,0.0017],[336,0.0018],[337,0.0017],[339,0.0018],[342,0.0016],[343,0.0017],[347,0.0016],[348,0.0017],[350,0.0011],[352,0.0017],[360,0.0017]]
       }
       CTRL_L_nose_wrinkleUpper {type=MovieSceneScriptingFloatChannel, default=0, const [1,492]}
       CTRL_R_nose_wrinkleUpper {type=MovieSceneScriptingFloatChannel, default=0, const [1,492]}
       CTRL_R_eye_squintInner {
        type=MovieSceneScriptingFloatChannel
        default=0
        thinned_from=492
        keys [[-131,0.0073],[-130,0.0052],[-129,0.0041],[-128,0.0043],[-126,0.0028],[-125,0.0022],[-124,0.0027],[-123,0.0019],[-122,0.0017],[-121,0.0004],[-120,0.003],[-119,0.003],[-117,0.0034],[-116,0.0052],[-115,0.006],[-114,0.0061],[-113,0.0054],[-112,0.0051],[-111,0.0055],[-110,0.0046],[-109,0.004],[-107,0.0054],[-106,0.0055],[-104,0.0054],[-103,0.005],[-101,0.003],[-100,0.0022],[-99,0.0029],[-97,0.0037],[-93,0.0031],[-92,0.0028],[-91,0.0028],[-90,0.0034],[-88,0.0041],[-87,0.0051],[-86,0.0069],[-84,0.0096],[-83,0.0106],[-82,0.0114],[-81,0.0129],[-80,0.0154],[-79,0.0156],[-77,0.0187],[-76,0.0194],[-75,0.0198],[-74,0.0206],[-73,0.022],[-72,0.0224],[-71,0.0221],[-70,0.0238],[-68,0.0264],[-67,0.0266],[-66,0.0257],[-65,0.0255],[-64,0.0264],[-63,0.0281],[-62,0.0294],[-61,0.0298],[-60,0.0299],[-59,0.031],[-58,0.0305],[-57,0.0322],[-56,0.0355],[-55,0.0355],[-54,0.0333],[-53,0.0313],[-52,0.0306],[-51,0.0328],[-50,0.0288],[-49,0.0283],[-48,0.0293],[-47,0.03],[-46,0.0293],[-45,0.0253],[-43,0.0259],[-42,0.0241],[-41,0.0258],[-40,0.0294],[-38,0.0283],[-37,0.0299],[-36,0.0321],[-34,0.0461],[-33,0.05],[-32,0.0551],[-31,0.0586],[-30,0.0762],[-29,0.0704],[-28,0.0593],[-27,0.0666],[-26,0.0653],[-25,0.0484],[-24,0.0425],[-23,0.025],[-22,0.0164],[-21,0.0161],[-20,0.022],[-19,0.0339],[-18,0.036],[-17,0.0288],[-16,0.0239],[-15,0.0236],[-14,0.0214],[-13,0.0221],[-12,0.0278],[-11,0.0293],[-10,0.0318],[-9,0.0358],[-8,0.0363],[-7,0.0392],[-6,0.0406],[-5,0.0423],[-4,0.041],[-3,0.0412],[-2,0.044],[-1,0.0449],[0,0.0466],[1,0.0495],[2,0.0562],[3,0.0578],[4,0.0679],[5,0.0892],[6,0.1519],[7,0.1698],[8,0.1724],[9,0.1783],[10,0.1815],[11,0.1755],[12,0.167],[13,0.17],[15,0.1994],[16,0.2085],[17,0.2149],[18,0.227],[19,0.2256],[20,0.2319],[21,0.2399],[22,0.2348],[23,0.2336],[24,0.2302],[25,0.2282],[26,0.2306],[27,0.2347],[28,0.2356],[29,0.2325],[30,0.2328],[31,0.2259],[32,0.2101],[33,0.2133],[34,0.2151],[35,0.215],[36,0.216],[37,0.2243],[39,0.2328],[40,0.2328],[41,0.219],[42,0.2066],[43,0.199],[44,0.1941],[45,0.194],[46,0.1804],[47,0.1433],[48,0.1015],[49,0.0931],[50,0.0982],[51,0.0869],[52,0.087],[53,0.0925],[54,0.1072],[55,0.1213],[56,0.1022],[57,0.1021],[58,0.1225],[59,0.1127],[60,0.1086],[61,0.1],[62,0.0931],[65,0.0931],[66,0.0911],[67,0.0833],[68,0.0739],[69,0.0728],[70,0.078],[71,0.083],[72,0.0908],[73,0.0972],[74,0.1154],[75,0.1189],[76,0.1199],[77,0.1296],[78,0.1302],[79,0.1396],[80,0.1475],[81,0.1519],[82,0.1454],[83,0.1407],[84,0.1393],[85,0.1336],[86,0.1265],[87,0.1298],[88,0.1318],[89,0.1261],[90,0.122],[91,0.1164],[92,0.1057],[93,0.1064],[94,0.1079],[95,0.1066],[96,0.1076],[97,0.1139],[98,0.1157],[99,0.109],[100,0.1095],[101,0.1078],[102,0.1056],[103,0.0891],[104,0.0951],[105,0.0956],[106,0.0935],[107,0.0878],[108,0.0927],[109,0.0892],[110,0.0853],[111,0.0865],[112,0.0814],[113,0.0799],[114,0.0868],[115,0.0902],[116,0.0777],[117,0.0649],[118,0.0571],[119,0.0481],[120,0.0459],[121,0.0672],[122,0.0898],[123,0.1026],[124,0.1082],[125,0.1141],[126,0.1258],[127,0.1236],[128,0.1229],[129,0.1267],[130,0.1237],[131,0.1201],[132,0.1169],[133,0.1112],[134,0.1103],[135,0.0934],[136,0.0551],[137,0.0428],[138,0.0816],[139,0.1305],[140,0.1417],[142,0.1252],[143,0.1274],[144,0.117],[145,0.1041],[146,0.1041],[147,0.1082],[148,0.1107],[149,0.1111],[150,0.1144],[151,0.1218],[152,0.1227],[153,0.121],[154,0.1273],[155,0.1263],[156,0.1171],[157,0.122],[158,0.1162],[159,0.1086],[160,0.1108],[161,0.1216],[162,0.1177],[163,0.1085],[164,0.105],[165,0.0904],[166,0.0646],[167,0.0614],[168,0.059],[169,0.0621],[170,0.0696],[171,0.0649],[172,0.0514],[173,0.0413],[174,0.0347],[175,0.0288],[176,0.0239],[177,0.0233],[178,0.0215],[180,0.0186],[181,0.0198],[182,0.0196],[183,0.0176],[184,0.0191],[185,0.0197],[186,0.0184],[187,0.0172],[188,0.0171],[189,0.0158],[190,0.0141],[191,0.0137],[192,0.0112],[193,0.011],[194,0.0103],[195,0.0086],[197,0.0042],[198,0.0026],[200,0.003],[201,0.0019],[203,0.0108],[204,0.0115],[205,0.0113],[206,0.0105],[208,0.0118],[209,0.0123],[210,0.014],[211,0.0149],[212,0.0144],[213,0.0137],[214,0.0127],[215,0.0128],[216,0.0138],[217,0.0136],[218,0.0138],[220,0.0148],[222,0.0148],[223,0.0151],[225,0.0142],[226,0.0143],[227,0.0149],[228,0.0147],[229,0.0137],[230,0.0133],[231,0.0135],[233,0.0132],[234,0.0137],[235,0.0136],[236,0.0193],[237,0.0165],[238,0.0131],[239,0.0101],[240,0.0125],[241,0.0175],[242,0.0183],[243,0.0214],[244,0.0226],[245,0.0166],[246,0.0094],[247,0.0073],[248,0.0073],[249,0.0069],[250,0.0059],[251,0.0073],[252,0.0083],[253,0.0076],[254,0.0093],[255,0.01],[257,0.0058],[258,0.0054],[259,0.0053],[260,0.007],[261,0.0122],[262,0.0146],[263,0.0302],[264,0.0329],[265,0.0315],[266,0.0244],[268,0.0127],[269,0.0103],[270,0.0101],[271,0.0128],[272,0.0083],[273,0.0048],[274,0.0031],[275,0.0029],[277,0.0039],[279,0.0043],[280,0.005],[284,0.005],[285,0.0052],[286,0.0032],[287,0.0031],[289,0.0038],[291,0.0029],[292,0.0057],[293,0.011],[295,0.0143],[296,0.0121],[297,0.0116],[298,0.0086],[299,0.0083],[300,0.0096],[301,0.0096],[302,0.0093],[303,0.0092],[305,0.008],[306,0.0083],[307,0.0095],[308,0.0103],[309,0.0109],[310,0.0111],[311,0.0107],[314,0.0105],[315,0.0101],[317,0.0101],[318,0.0098],[320,0.0106],[321,0.0109],[322,0.0102],[323,0.009],[324,0.0095],[325,0.0104],[326,0.0105],[327,0.0101],[328,0.0105],[329,0.0112],[330,0.0108],[331,0.0109],[332,0.0114],[333,0.011],[334,0.0111],[335,0.0113],[336,0.0112],[337,0.0109],[339,0.0112],[341,0.0109],[342,0.0109],[343,0.0116],[346,0.0111],[347,0.0103],[348,0.0105],[349,0.0111],[350,0.011],[351,0.0115],[352,0.0124],[353,0.013],[355,0.0134],[356,0.0134],[358,0.0127],[360,0.0127]]
       }
       CTRL_R_eye_cheekRaise {
        type=MovieSceneScriptingFloatChannel
        default=0
        thinned_from=492
        keys [[-131,0.0063],[-129,0.0052],[-128,0.0044],[-126,0.0034],[-125,0.0032],[-124,0.0034],[-122,0.0032],[-121,0.0028],[-120,0.0034],[-119,0.0036],[-118,0.0033],[-117,0.0033],[-115,0.0042],[-114,0.0044],[-112,0.0041],[-110,0.004],[-109,0.0037],[-108,0.0038],[-107,0.0043],[-105,0.0046],[-104,0.0046],[-103,0.0042],[-102,0.0042],[-99,0.0035],[-96,0.0041],[-94,0.0044],[-93,0.0044],[-91,0.0038],[-86,0.0059],[-85,0.0062],[-84,0.0061],[-80,0.0044],[-77,0.0041],[-76,0.0043],[-74,0.0042],[-67,0.0036],[-65,0.0036],[-63,0.0032],[-61,0.0033],[-60,0.0036],[-59,0.0046],[-58,0.0064],[-57,0.007],[-56,0.0083],[-55,0.0093],[-54,0.0093],[-53,0.0088],[-52,0.0086],[-51,0.0089],[-49,0.0086],[-48,0.008],[-47,0.0084],[-44,0.0079],[-41,0.0058],[-40,0.0055],[-39,0.0049],[-38,0.0047],[-37,0.004],[-36,0.0038],[-34,0.0051],[-33,0.0065],[-32,0.0082],[-31,0.0089],[-30,0.0123],[-29,0.0122],[-26,0.0113],[-25,0.0097],[-24,0.0091],[-23,0.007],[-22,0.0059],[-21,0.0057],[-20,0.0062],[-18,0.0051],[-17,0.0044],[-16,0.0045],[-15,0.005],[-14,0.005],[-13,0.0053],[-10,0.0076],[-9,0.0083],[-8,0.0093],[-7,0.0111],[-6,0.0125],[-5,0.0142],[-4,0.0139],[-3,0.0144],[-2,0.0142],[-1,0.0152],[0,0.0158],[1,0.0171],[2,0.0193],[3,0.019],[4,0.019],[5,0.0185],[6,0.0193],[8,0.0183],[9,0.0175],[10,0.0177],[12,0.0171],[13,0.0172],[14,0.0177],[15,0.0188],[16,0.0196],[18,0.0206],[19,0.0227],[20,0.0239],[22,0.0245],[23,0.0242],[24,0.0235],[25,0.0232],[26,0.0224],[27,0.0226],[28,0.0225],[29,0.0213],[30,0.0206],[31,0.0193],[32,0.0185],[33,0.0167],[34,0.0162],[35,0.0161],[36,0.0157],[37,0.0162],[39,0.0195],[40,0.02],[41,0.0183],[42,0.0169],[43,0.0158],[44,0.015],[45,0.0147],[46,0.0129],[47,0.0107],[48,0.0069],[49,0.0054],[50,0.0046],[51,0.0044],[52,0.0049],[53,0.0058],[54,0.0055],[56,0.004],[58,0.0021],[59,0.0016],[60,0.0015],[62,0.0023],[68,0.0024],[69,0.0022],[70,0.0023],[71,0.002],[73,0.001],[78,0.0008],[79,0.0005],[83,0.0005],[85,0],[99,0],[101,0.0004],[102,0.0004],[103,0],[175,0],[177,0.0003],[178,0.0006],[181,0.0002],[182,0.0003],[184,0.0001],[187,0.0012],[192,0.0013],[193,0.0024],[194,0.0048],[195,0.0049],[196,0.0043],[197,0.0031],[198,0.0024],[200,0.0024],[201,0.0028],[203,0.0031],[204,0.0037],[206,0.0041],[207,0.0041],[211,0.005],[215,0.0047],[219,0.005],[220,0.0055],[226,0.0054],[228,0.0059],[232,0.0054],[235,0.0043],[237,0.0063],[238,0.0064],[239,0.0057],[240,0.0056],[241,0.0061],[243,0.0061],[244,0.0055],[247,0.0047],[248,0.004],[249,0.0041],[250,0.0038],[251,0.0042],[254,0.0038],[256,0.0041],[257,0.0028],[258,0.002],[259,0.0022],[261,0.0048],[262,0.0059],[263,0.0067],[264,0.0071],[265,0.0079],[267,0.0078],[268,0.0075],[270,0.0076],[272,0.0064],[273,0.0055],[274,0.0044],[275,0.0039],[276,0.0038],[280,0.0037],[283,0.0042],[285,0.0041],[286,0.0033],[288,0.0035],[289,0.004],[290,0.004],[291,0.0037],[292,0.0039],[294,0.0059],[295,0.0065],[297,0.0067],[298,0.0062],[299,0.006],[300,0.0063],[301,0.0062],[302,0.0059],[303,0.0058],[305,0.0053],[310,0.0055],[311,0.0052],[317,0.0052],[318,0.005],[319,0.005],[321,0.0055],[323,0.005],[324,0.005],[326,0.0056],[327,0.0054],[335,0.0053],[336,0.0055],[340,0.0051],[341,0.0049],[352,0.005],[353,0.0052],[357,0.0051],[358,0.0052],[359,0.0049],[360,0.0049]]
       }
       CTRL_C_eyelidsReduce {type=MovieSceneScriptingFloatChannel, default=0, const [0,492]}
       CTRL_R_brow_lateral {
        type=MovieSceneScriptingFloatChannel
        default=0
        thinned_from=492
        keys [[-131,0.006],[-130,0.0061],[-129,0.0063],[-126,0.0059],[-122,0.0059],[-121,0.0053],[-120,0.0055],[-119,0.0059],[-118,0.0056],[-115,0.006],[-114,0.0064],[-111,0.0064],[-109,0.0067],[-108,0.006],[-107,0.0056],[-104,0.0053],[-102,0.0046],[-99,0.0049],[-92,0.0048],[-88,0.0055],[-85,0.0055],[-81,0.0061],[-80,0.0067],[-79,0.0067],[-78,0.0071],[-77,0.0072],[-75,0.0064],[-74,0.0054],[-72,0.0047],[-69,0.0042],[-68,0.0044],[-63,0.004],[-60,0.0034],[-58,0.0035],[-57,0.0033],[-55,0.0024],[-52,0.0021],[-51,0.0023],[-50,0.0027],[-46,0.003],[-45,0.0033],[-44,0.0033],[-41,0.0043],[-39,0.0058],[-38,0.0059],[-37,0.0059],[-35,0.0047],[-34,0.0045],[-33,0.0045],[-31,0.0039],[-30,0.0028],[-28,0.0025],[-24,0.0029],[-23,0.0029],[-20,0.0034],[-19,0.0046],[-18,0.0153],[-17,0.0315],[-16,0.0536],[-15,0.0657],[-13,0.08],[-12,0.0952],[-11,0.1109],[-10,0.1262],[-9,0.1444],[-8,0.1582],[-7,0.1672],[-6,0.1751],[-5,0.1785],[-4,0.1831],[-3,0.185],[-2,0.192],[-1,0.1945],[0,0.2002],[1,0.2071],[2,0.2169],[3,0.2159],[4,0.2111],[5,0.2137],[6,0.2207],[7,0.2173],[8,0.2154],[9,0.2109],[11,0.2082],[13,0.2065],[14,0.2064],[15,0.2053],[16,0.203],[17,0.1994],[18,0.1993],[19,0.1976],[20,0.1967],[21,0.195],[22,0.1939],[23,0.1934],[24,0.1915],[25,0.1876],[26,0.1825],[28,0.1746],[29,0.1703],[30,0.1664],[31,0.1629],[32,0.1598],[33,0.1598],[34,0.1611],[35,0.161],[36,0.1598],[37,0.1557],[38,0.1567],[39,0.1597],[40,0.1641],[41,0.1634],[42,0.1597],[43,0.1566],[44,0.1547],[45,0.1542],[46,0.1405],[47,0.1073],[48,0.0577],[49,0.0318],[50,0.0186],[51,0.0088],[53,0.0011],[60,0.0007],[62,0.0011],[63,0.0011],[64,0.0015],[66,0.0015],[67,0.0014],[68,0.0015],[71,0.0008],[72,0.0007],[75,0.0013],[77,0.0009],[79,0.0009],[80,0.0011],[81,0.0016],[84,0.0019],[87,0.002],[89,0.0023],[91,0.0023],[92,0.002],[99,0.0018],[102,0.0016],[103,0.002],[109,0.002],[110,0.0023],[111,0.0024],[112,0.002],[113,0.002],[114,0.0017],[115,0.0019],[116,0.0018],[117,0.002],[118,0.002],[119,0.0017],[121,0.0016],[122,0.0019],[124,0.002],[125,0.0017],[127,0.0017],[128,0.0022],[130,0.0022],[132,0.0017],[134,0.0017],[135,0.0022],[137,0.0027],[139,0.0019],[140,0.0012],[143,0.0008],[145,0.0009],[146,0.0013],[151,0.0015],[154,0.0009],[155,0.001],[157,0.0016],[158,0.0014],[159,0.0009],[162,0.0011],[163,0.0011],[166,0.0013],[167,0.0011],[169,0.0013],[173,0.0012],[174,0.001],[175,0.0011],[176,0.0015],[178,0.0016],[180,0.0011],[181,0.0006],[182,0.0007],[184,0.0013],[186,0.0016],[188,0.0015],[190,0.0019],[191,0.0024],[193,0.0024],[194,0.0028],[195,0.0028],[196,0.0024],[197,0.0022],[199,0.0035],[202,0.0037],[203,0.0033],[204,0.0033],[205,0.0028],[215,0.0025],[219,0.0026],[223,0.0029],[226,0.0025],[228,0.0028],[233,0.0025],[234,0.0027],[235,0.0031],[237,0.0052],[239,0.0067],[240,0.0069],[241,0.0063],[242,0.0062],[244,0.0055],[246,0.0061],[248,0.0075],[249,0.0076],[251,0.0069],[252,0.0069],[253,0.0067],[255,0.0056],[256,0.0059],[260,0.0056],[261,0.007],[262,0.01],[263,0.0185],[265,0.0194],[266,0.0176],[267,0.0145],[269,0.0074],[270,0.0069],[271,0.005],[272,0.0048],[276,0.0059],[278,0.0059],[280,0.0063],[281,0.0062],[282,0.0059],[284,0.0059],[285,0.006],[286,0.0056],[289,0.0059],[290,0.0057],[291,0.005],[292,0.0048],[294,0.0061],[295,0.006],[296,0.0052],[297,0.0051],[299,0.0042],[300,0.0032],[301,0.0031],[302,0.0028],[305,0.0023],[308,0.0027],[311,0.0022],[313,0.0022],[314,0.0024],[318,0.0022],[322,0.0024],[323,0.0024],[324,0.0025],[326,0.0024],[327,0.0022],[337,0.0027],[340,0.0024],[342,0.0021],[345,0.0024],[347,0.0022],[349,0.0024],[352,0.0028],[358,0.0025],[360,0.0024]]
       }
       CTRL_L_brow_lateral {
        type=MovieSceneScriptingFloatChannel
        default=0
        thinned_from=492
        keys [[-131,0.0075],[-129,0.0084],[-128,0.0084],[-127,0.0079],[-126,0.0078],[-124,0.0079],[-123,0.0076],[-122,0.0076],[-121,0.007],[-120,0.0071],[-119,0.0074],[-117,0.0073],[-115,0.0077],[-114,0.0081],[-109,0.0081],[-107,0.0073],[-104,0.0072],[-102,0.0067],[-97,0.0074],[-95,0.0072],[-94,0.0069],[-90,0.0068],[-88,0.0072],[-82,0.0071],[-77,0.0068],[-76,0.0066],[-74,0.0053],[-72,0.005],[-69,0.0047],[-65,0.0047],[-64,0.0045],[-60,0.0047],[-58,0.0049],[-53,0.0039],[-51,0.0038],[-49,0.004],[-47,0.0045],[-41,0.0051],[-39,0.0059],[-37,0.006],[-35,0.0051],[-33,0.005],[-31,0.005],[-27,0.0043],[-24,0.0043],[-22,0.0049],[-21,0.0056],[-20,0.0056],[-19,0.0059],[-18,0.0147],[-17,0.0295],[-16,0.0509],[-15,0.0631],[-14,0.0701],[-13,0.0766],[-12,0.0902],[-10,0.1186],[-9,0.1357],[-8,0.1489],[-7,0.1573],[-6,0.1646],[-5,0.1671],[-4,0.1716],[-3,0.1727],[-2,0.1794],[-1,0.1818],[0,0.1876],[1,0.1942],[2,0.2031],[3,0.202],[4,0.1973],[5,0.2005],[6,0.2081],[7,0.2052],[8,0.2036],[9,0.1998],[11,0.1973],[14,0.1953],[15,0.193],[16,0.1904],[17,0.1873],[18,0.1882],[19,0.1858],[20,0.1853],[21,0.1835],[23,0.1826],[24,0.1805],[25,0.1771],[26,0.1721],[28,0.1638],[29,0.16],[31,0.1537],[32,0.1509],[33,0.1513],[34,0.153],[35,0.153],[36,0.1519],[37,0.1476],[38,0.1481],[39,0.1504],[40,0.1546],[41,0.1546],[43,0.1484],[44,0.1467],[45,0.1464],[46,0.1337],[47,0.1026],[48,0.056],[49,0.0318],[50,0.0196],[51,0.0107],[53,0.0041],[56,0.0039],[58,0.0035],[61,0.0036],[63,0.0039],[64,0.0045],[66,0.0044],[68,0.0041],[71,0.0037],[72,0.0038],[73,0.0044],[75,0.0047],[78,0.0046],[80,0.005],[81,0.0055],[82,0.0055],[84,0.0059],[86,0.0056],[88,0.0059],[89,0.0063],[91,0.0067],[92,0.0063],[95,0.0065],[99,0.0058],[100,0.0055],[102,0.0055],[103,0.006],[105,0.006],[108,0.0057],[110,0.0063],[111,0.0063],[112,0.006],[114,0.0059],[115,0.0061],[117,0.0059],[119,0.0053],[120,0.0051],[123,0.0059],[124,0.006],[125,0.0059],[127,0.0061],[128,0.0064],[130,0.0063],[131,0.0061],[132,0.0062],[134,0.0059],[136,0.006],[137,0.0058],[138,0.0053],[142,0.0046],[145,0.0046],[147,0.0052],[152,0.0053],[154,0.0047],[155,0.0046],[157,0.0053],[158,0.0052],[159,0.0047],[162,0.0049],[163,0.0047],[165,0.0049],[167,0.0046],[169,0.0052],[170,0.0049],[173,0.005],[175,0.0046],[176,0.0049],[178,0.005],[180,0.0044],[181,0.004],[186,0.0047],[189,0.0044],[190,0.0044],[191,0.0049],[192,0.0047],[195,0.0044],[196,0.0041],[197,0.0043],[199,0.0057],[200,0.0059],[204,0.0058],[205,0.0053],[207,0.0049],[213,0.0046],[215,0.0044],[219,0.0046],[220,0.005],[226,0.0047],[228,0.005],[229,0.005],[230,0.0047],[232,0.0046],[234,0.005],[238,0.0069],[239,0.0078],[240,0.0083],[241,0.0077],[242,0.0076],[244,0.0066],[246,0.0071],[248,0.0088],[249,0.0092],[251,0.0086],[252,0.0086],[255,0.0073],[256,0.0076],[257,0.0073],[260,0.0071],[261,0.0081],[262,0.0102],[263,0.0156],[265,0.0158],[266,0.0148],[267,0.0129],[268,0.0106],[269,0.0085],[270,0.0082],[271,0.0072],[272,0.0071],[273,0.0072],[275,0.008],[276,0.0082],[278,0.0081],[280,0.0084],[281,0.0083],[283,0.0077],[285,0.0078],[286,0.0077],[289,0.008],[290,0.0078],[291,0.0073],[292,0.0072],[294,0.0082],[295,0.0082],[296,0.0077],[297,0.0076],[299,0.0064],[300,0.0053],[301,0.0053],[302,0.0049],[303,0.0048],[304,0.0044],[308,0.0047],[317,0.0043],[318,0.0042],[321,0.0046],[326,0.005],[329,0.0048],[332,0.0052],[334,0.0049],[340,0.005],[342,0.0047],[345,0.0049],[349,0.0044],[357,0.0047],[360,0.0046]]
       }
       CTRL_R_brow_down {
        type=MovieSceneScriptingFloatChannel
        default=0
        thinned_from=492
        keys [[-131,0.0067],[-129,0.0062],[-126,0.0045],[-125,0.0046],[-124,0.0051],[-122,0.0052],[-121,0.0045],[-120,0.0045],[-119,0.005],[-117,0.0049],[-114,0.006],[-110,0.0065],[-108,0.0063],[-107,0.0068],[-106,0.0069],[-105,0.0072],[-104,0.0073],[-102,0.0067],[-100,0.0065],[-99,0.0063],[-97,0.0067],[-95,0.0067],[-94,0.0064],[-87,0.0058],[-85,0.0064],[-84,0.0065],[-81,0.0061],[-80,0.0066],[-79,0.0066],[-78,0.0071],[-75,0.0118],[-74,0.0126],[-73,0.0125],[-71,0.014],[-70,0.0145],[-69,0.0143],[-67,0.0145],[-65,0.0144],[-63,0.0151],[-60,0.0148],[-59,0.0149],[-57,0.0161],[-56,0.0175],[-55,0.0192],[-54,0.02],[-52,0.0195],[-51,0.0183],[-50,0.018],[-49,0.018],[-48,0.0175],[-47,0.0166],[-44,0.0165],[-43,0.0158],[-42,0.0135],[-41,0.0116],[-40,0.01],[-39,0.0095],[-38,0.0094],[-37,0.0114],[-36,0.0129],[-35,0.0148],[-33,0.0167],[-32,0.0179],[-31,0.0183],[-30,0.021],[-28,0.0209],[-27,0.0218],[-26,0.0233],[-25,0.0236],[-24,0.0234],[-23,0.0182],[-22,0.0124],[-21,0.0075],[-20,0.0079],[-19,0.0105],[-18,0.0188],[-17,0.0283],[-16,0.0404],[-15,0.0467],[-13,0.0556],[-11,0.0776],[-10,0.0896],[-9,0.1027],[-8,0.1133],[-7,0.1206],[-6,0.1275],[-4,0.1342],[-3,0.1359],[-2,0.1412],[-1,0.1427],[0,0.1479],[1,0.1537],[2,0.1624],[3,0.1616],[4,0.1587],[5,0.1617],[6,0.1735],[7,0.1747],[8,0.1741],[9,0.1726],[10,0.172],[12,0.1695],[13,0.1689],[15,0.1708],[16,0.1695],[17,0.1671],[18,0.1686],[19,0.1707],[21,0.1705],[22,0.1692],[23,0.1694],[24,0.1688],[25,0.1648],[26,0.1611],[27,0.158],[30,0.1515],[32,0.1439],[33,0.1431],[34,0.1448],[35,0.1447],[36,0.1434],[37,0.1414],[38,0.1447],[39,0.1499],[40,0.153],[41,0.1492],[42,0.1446],[44,0.1409],[45,0.1401],[46,0.1243],[47,0.0974],[48,0.0572],[49,0.038],[50,0.0263],[51,0.0177],[53,0.0115],[54,0.0112],[58,0.0064],[59,0.0056],[61,0.0053],[62,0.0061],[63,0.0061],[64,0.0068],[66,0.007],[69,0.0068],[71,0.0072],[72,0.0077],[73,0.0083],[76,0.0097],[79,0.0096],[80,0.01],[81,0.0107],[83,0.0115],[84,0.0111],[85,0.011],[86,0.0103],[87,0.0104],[88,0.0103],[89,0.0105],[92,0.0103],[95,0.0108],[96,0.0108],[97,0.0106],[99,0.0108],[100,0.0114],[101,0.0116],[102,0.0115],[104,0.0108],[105,0.0107],[107,0.0098],[108,0.0098],[109,0.0099],[111,0.0086],[112,0.0074],[113,0.0066],[114,0.0056],[117,0.0044],[119,0.0031],[121,0.003],[122,0.0033],[123,0.0029],[124,0.0026],[125,0.0016],[127,0.0013],[128,0.0018],[129,0.0019],[131,0.0013],[132,0.0008],[133,0.0008],[135,0.0016],[138,0.0005],[139,0.0004],[140,0.001],[141,0.0008],[142,0.0013],[144,0.0014],[145,0.0007],[146,0.0003],[149,0.0004],[150,0.0002],[153,0.001],[154,0.001],[155,0.0013],[156,0.0013],[158,0.0009],[163,0.0019],[167,0.0019],[168,0.0017],[169,0.0022],[171,0.0026],[173,0.0022],[174,0.0016],[175,0.0014],[177,0.0018],[178,0.0022],[180,0.0017],[181,0.0013],[182,0.0014],[184,0.0013],[185,0.0014],[187,0.002],[191,0.0024],[192,0.0029],[193,0.004],[194,0.0049],[195,0.0046],[197,0.002],[198,0.0011],[199,0.0005],[201,0.0004],[202,0.0006],[203,0.0011],[204,0.0018],[205,0.002],[206,0.0025],[207,0.0027],[210,0.0041],[211,0.0045],[214,0.0042],[219,0.0048],[223,0.0058],[225,0.0059],[226,0.0058],[227,0.006],[233,0.005],[234,0.005],[235,0.0045],[237,0.005],[238,0.0049],[239,0.0043],[240,0.0043],[241,0.0047],[242,0.0048],[243,0.0055],[245,0.0058],[247,0.0048],[248,0.0041],[249,0.0039],[250,0.0033],[251,0.0037],[253,0.0029],[254,0.0023],[255,0.002],[257,0.0029],[259,0.0033],[260,0.0048],[261,0.0074],[262,0.014],[263,0.0355],[264,0.039],[265,0.0455],[266,0.042],[268,0.0211],[269,0.0129],[270,0.0118],[271,0.0071],[272,0.0055],[274,0.004],[278,0.0035],[280,0.0043],[281,0.0043],[282,0.0038],[284,0.0037],[285,0.0039],[286,0.0029],[287,0.0029],[288,0.0032],[289,0.0041],[290,0.0042],[291,0.0039],[292,0.0041],[293,0.0051],[294,0.0056],[295,0.0053],[296,0.0048],[298,0.0046],[299,0.0048],[300,0.0057],[301,0.0058],[302,0.0072],[306,0.0078],[307,0.0082],[308,0.0084],[310,0.0082],[313,0.0087],[314,0.0094],[315,0.0094],[316,0.0093],[317,0.0089],[319,0.0087],[320,0.0091],[321,0.0093],[323,0.009],[326,0.0097],[327,0.0093],[329,0.0098],[330,0.0104],[332,0.0107],[335,0.0108],[337,0.0111],[341,0.01],[342,0.01],[343,0.0103],[345,0.0105],[350,0.01],[352,0.0106],[353,0.0112],[355,0.0119],[358,0.0123],[359,0.012],[360,0.012]]
       }
       CTRL_L_brow_down {
        type=MovieSceneScriptingFloatChannel
        default=0
        thinned_from=492
        keys [[-131,0.0087],[-129,0.0079],[-128,0.0072],[-126,0.0064],[-124,0.0068],[-122,0.0068],[-121,0.006],[-120,0.0064],[-119,0.007],[-118,0.0068],[-116,0.0071],[-114,0.0078],[-110,0.0081],[-109,0.0079],[-108,0.008],[-107,0.0083],[-104,0.0086],[-99,0.0077],[-95,0.008],[-94,0.0077],[-93,0.0076],[-91,0.0069],[-90,0.0071],[-85,0.0091],[-82,0.0095],[-81,0.0091],[-80,0.0094],[-78,0.0094],[-75,0.0122],[-74,0.0126],[-73,0.0123],[-70,0.0137],[-67,0.0139],[-64,0.014],[-63,0.0142],[-59,0.0142],[-58,0.0151],[-57,0.0156],[-55,0.0187],[-54,0.0194],[-52,0.019],[-51,0.0184],[-49,0.0182],[-47,0.017],[-44,0.0169],[-43,0.016],[-42,0.014],[-41,0.0124],[-40,0.011],[-39,0.0104],[-38,0.0103],[-37,0.012],[-36,0.0129],[-35,0.0142],[-34,0.0147],[-33,0.0157],[-32,0.0168],[-31,0.0174],[-30,0.0209],[-29,0.0209],[-28,0.0212],[-27,0.0219],[-26,0.0232],[-25,0.0234],[-24,0.0232],[-23,0.0184],[-22,0.0127],[-21,0.0081],[-20,0.0086],[-19,0.011],[-18,0.0192],[-17,0.0283],[-16,0.0408],[-15,0.0476],[-14,0.0521],[-13,0.0561],[-12,0.0656],[-11,0.0754],[-10,0.0862],[-9,0.0982],[-8,0.108],[-7,0.1148],[-6,0.1209],[-5,0.1234],[-4,0.1266],[-3,0.1275],[-2,0.1324],[-1,0.134],[0,0.1395],[1,0.1453],[2,0.1532],[3,0.1522],[4,0.1492],[5,0.1528],[6,0.1643],[7,0.1656],[8,0.1653],[9,0.1645],[10,0.1643],[12,0.1621],[13,0.1615],[15,0.1621],[16,0.1605],[17,0.1583],[18,0.1605],[19,0.1618],[20,0.1626],[21,0.1626],[23,0.1621],[24,0.161],[26,0.1537],[27,0.1505],[28,0.1483],[30,0.1452],[32,0.1386],[33,0.1381],[34,0.1401],[35,0.14],[36,0.1388],[37,0.1367],[38,0.1394],[39,0.1436],[40,0.1464],[41,0.1436],[42,0.1394],[43,0.1372],[44,0.1356],[45,0.1351],[46,0.1206],[47,0.0954],[48,0.0578],[49,0.0397],[50,0.0288],[51,0.0201],[52,0.017],[53,0.0142],[54,0.0139],[55,0.0129],[58,0.0092],[59,0.0085],[60,0.0084],[61,0.008],[62,0.0089],[63,0.0089],[64,0.0094],[65,0.0094],[66,0.0096],[70,0.0094],[72,0.0096],[75,0.0114],[76,0.0118],[80,0.0118],[81,0.0126],[82,0.0127],[83,0.0132],[84,0.0129],[85,0.0129],[86,0.0122],[87,0.0124],[88,0.0123],[89,0.0125],[90,0.012],[96,0.0128],[98,0.0125],[99,0.0127],[100,0.0133],[101,0.0134],[102,0.0134],[104,0.0128],[105,0.0128],[106,0.0119],[107,0.0114],[108,0.0113],[109,0.0116],[111,0.0106],[112,0.0093],[113,0.0087],[114,0.0074],[116,0.0068],[119,0.0061],[120,0.0064],[121,0.0063],[122,0.0063],[123,0.0056],[124,0.0056],[125,0.0048],[127,0.0044],[129,0.0047],[130,0.0046],[132,0.0038],[133,0.0037],[134,0.0044],[135,0.0048],[136,0.0045],[137,0.0037],[139,0.0029],[140,0.0032],[141,0.0029],[142,0.0031],[143,0.0031],[144,0.0037],[145,0.0035],[146,0.003],[147,0.0032],[148,0.003],[149,0.0031],[150,0.0029],[151,0.0033],[154,0.0033],[155,0.0037],[157,0.0037],[158,0.0034],[159,0.0037],[160,0.0036],[163,0.0041],[166,0.0038],[167,0.004],[168,0.0037],[169,0.0044],[171,0.0053],[173,0.0051],[174,0.0043],[175,0.004],[177,0.0045],[178,0.0051],[181,0.0036],[182,0.004],[185,0.004],[187,0.0046],[191,0.0045],[192,0.0051],[194,0.0072],[195,0.0074],[196,0.007],[198,0.0054],[199,0.005],[202,0.0046],[203,0.0051],[204,0.006],[205,0.0062],[211,0.0082],[213,0.0081],[217,0.0075],[219,0.0077],[220,0.0083],[222,0.0085],[226,0.0086],[228,0.0089],[231,0.0085],[234,0.0085],[236,0.0067],[239,0.0055],[240,0.0057],[241,0.0066],[242,0.0068],[243,0.0077],[245,0.0078],[246,0.0072],[247,0.007],[248,0.0064],[249,0.0067],[250,0.0059],[251,0.0059],[252,0.0052],[254,0.0042],[255,0.0039],[256,0.0044],[259,0.0047],[260,0.0062],[261,0.0084],[262,0.0133],[263,0.029],[264,0.0319],[265,0.0369],[266,0.0341],[267,0.0255],[268,0.0174],[269,0.0117],[270,0.011],[271,0.0082],[272,0.0074],[273,0.0069],[274,0.0061],[277,0.0059],[278,0.0056],[281,0.0061],[282,0.0058],[285,0.0059],[286,0.0051],[287,0.0051],[288,0.0053],[289,0.0061],[290,0.0061],[291,0.0057],[292,0.0057],[293,0.0061],[294,0.0063],[295,0.006],[297,0.006],[298,0.0063],[299,0.0069],[300,0.008],[301,0.0081],[302,0.0089],[305,0.009],[308,0.0097],[311,0.0092],[312,0.0092],[314,0.0097],[315,0.0098],[318,0.0091],[319,0.009],[320,0.0095],[321,0.0097],[322,0.0095],[323,0.0091],[326,0.0102],[327,0.0098],[332,0.0106],[335,0.0107],[336,0.011],[337,0.011],[341,0.0099],[345,0.0106],[348,0.01],[350,0.0099],[352,0.0107],[353,0.0115],[355,0.012],[358,0.0121],[359,0.0119],[360,0.0119]]
       }
       CTRL_L_brow_raiseOut {
        type=MovieSceneScriptingFloatChannel
        default=0.7687
        thinned_from=492
        keys [[-131,0.5975],[-130,0.6191],[-129,0.619],[-128,0.6114],[-126,0.607],[-125,0.609],[-124,0.6049],[-123,0.5976],[-122,0.596],[-121,0.5958],[-120,0.6014],[-119,0.5842],[-118,0.5763],[-117,0.5737],[-116,0.5566],[-115,0.5527],[-114,0.557],[-113,0.5674],[-112,0.5682],[-111,0.5558],[-110,0.5376],[-109,0.5272],[-108,0.5224],[-107,0.519],[-106,0.5183],[-105,0.5126],[-103,0.5131],[-101,0.5397],[-100,0.5447],[-99,0.5323],[-98,0.5219],[-96,0.5128],[-95,0.5114],[-94,0.517],[-93,0.5162],[-92,0.5089],[-91,0.4934],[-89,0.4803],[-88,0.4817],[-87,0.4847],[-86,0.4814],[-85,0.4695],[-84,0.4502],[-83,0.4413],[-82,0.4402],[-81,0.4453],[-80,0.4372],[-79,0.4365],[-78,0.4295],[-77,0.4171],[-75,0.4061],[-74,0.4053],[-73,0.4145],[-71,0.4106],[-70,0.4026],[-69,0.401],[-67,0.4111],[-66,0.4098],[-65,0.4005],[-64,0.3948],[-63,0.3942],[-61,0.3975],[-60,0.3954],[-59,0.3892],[-58,0.3776],[-57,0.3777],[-56,0.3762],[-55,0.3728],[-54,0.3744],[-53,0.3807],[-51,0.3887],[-50,0.3861],[-49,0.3857],[-48,0.3925],[-47,0.4141],[-46,0.4133],[-45,0.4083],[-44,0.4062],[-43,0.4029],[-42,0.4146],[-41,0.42],[-40,0.4186],[-38,0.4205],[-37,0.4182],[-36,0.4096],[-35,0.3966],[-34,0.3934],[-33,0.3886],[-32,0.3933],[-31,0.3932],[-30,0.3869],[-29,0.3865],[-27,0.3798],[-26,0.3732],[-25,0.3814],[-24,0.3876],[-23,0.4836],[-22,0.6237],[-21,0.7883],[-20,0.8616],[-19,0.8561],[-18,0.7814],[-17,0.7152],[-16,0.678],[-15,0.6603],[-14,0.6446],[-13,0.6369],[-12,0.6182],[-10,0.5691],[-8,0.5367],[-7,0.518],[-6,0.498],[-5,0.4956],[-4,0.4948],[-3,0.4982],[-2,0.4934],[-1,0.4952],[0,0.4766],[1,0.4631],[2,0.4465],[3,0.4502],[4,0.4565],[5,0.4498],[6,0.4326],[7,0.4323],[8,0.433],[9,0.4352],[10,0.435],[11,0.4285],[12,0.4194],[13,0.4162],[14,0.4274],[15,0.4371],[16,0.4352],[17,0.4296],[18,0.4179],[19,0.4125],[20,0.4026],[21,0.3962],[22,0.4008],[23,0.4026],[25,0.4149],[26,0.4182],[27,0.425],[28,0.4287],[29,0.4225],[30,0.4137],[31,0.4173],[32,0.4309],[33,0.4367],[34,0.4356],[35,0.4358],[37,0.4408],[38,0.4313],[39,0.4208],[40,0.4179],[41,0.4234],[43,0.4253],[44,0.4338],[45,0.436],[46,0.5066],[47,0.5504],[48,0.6344],[49,0.6652],[50,0.7269],[51,0.7534],[52,0.7653],[53,0.7365],[54,0.7247],[55,0.6982],[56,0.6736],[57,0.6504],[58,0.642],[59,0.6462],[60,0.6476],[61,0.6436],[62,0.6152],[63,0.6124],[64,0.5679],[65,0.5679],[66,0.5487],[67,0.543],[68,0.5395],[69,0.5321],[70,0.5173],[72,0.4914],[73,0.4612],[74,0.4441],[76,0.4134],[77,0.4053],[79,0.4047],[80,0.408],[81,0.4056],[82,0.4001],[83,0.3921],[84,0.3928],[86,0.4052],[88,0.3991],[89,0.3931],[90,0.3926],[91,0.3852],[92,0.3822],[93,0.3807],[94,0.3759],[95,0.3685],[96,0.3652],[97,0.3696],[98,0.3749],[99,0.3701],[100,0.3617],[102,0.3656],[103,0.3572],[104,0.3595],[105,0.3598],[106,0.3782],[107,0.3905],[108,0.3895],[115,0.3906],[116,0.3915],[117,0.3934],[118,0.3973],[119,0.4029],[122,0.4244],[123,0.4334],[124,0.4436],[126,0.4701],[127,0.4865],[128,0.501],[129,0.5192],[130,0.5396],[131,0.5688],[132,0.6024],[133,0.642],[134,0.7084],[135,0.7883],[136,0.8804],[137,0.9476],[138,0.9989],[139,0.9815],[140,0.9449],[141,0.9274],[142,0.9038],[143,0.9154],[144,0.9365],[145,0.9459],[146,0.9584],[147,0.9533],[148,0.9504],[149,0.9349],[150,0.9236],[151,0.9061],[152,0.895],[155,0.8925],[156,0.8847],[157,0.8801],[159,0.862],[161,0.8568],[162,0.8518],[163,0.8495],[164,0.8446],[165,0.8468],[166,0.8397],[168,0.8419],[169,0.841],[171,0.8347],[172,0.8335],[173,0.8282],[174,0.8269],[175,0.8322],[176,0.8405],[177,0.8384],[178,0.8284],[179,0.8126],[180,0.801],[181,0.7969],[182,0.7968],[183,0.7924],[184,0.7775],[185,0.7539],[186,0.7345],[187,0.727],[189,0.7239],[190,0.7119],[191,0.6954],[192,0.6538],[193,0.6594],[194,0.8082],[195,0.8652],[196,0.9584],[197,0.9931],[198,1],[210,1],[211,0.9957],[212,0.994],[213,0.9851],[214,0.9822],[215,0.9733],[216,0.9619],[217,0.9435],[220,0.8956],[221,0.8836],[222,0.8762],[223,0.8654],[224,0.8564],[225,0.8493],[226,0.8478],[228,0.8357],[230,0.8202],[231,0.8199],[232,0.8175],[233,0.8336],[234,0.8356],[235,0.8542],[236,0.8768],[237,0.8944],[238,0.8911],[239,0.8507],[240,0.7908],[241,0.7284],[242,0.717],[243,0.7014],[244,0.6609],[245,0.5932],[246,0.5034],[247,0.4759],[248,0.4606],[249,0.4613],[250,0.4776],[251,0.5013],[252,0.5232],[254,0.5306],[255,0.5517],[256,0.5529],[257,0.5434],[258,0.5555],[259,0.5509],[260,0.4944],[261,0.429],[262,0.3585],[263,0.2912],[264,0.2797],[265,0.2533],[266,0.259],[267,0.2961],[268,0.3851],[269,0.4875],[270,0.5026],[271,0.6119],[272,0.6664],[273,0.7078],[274,0.753],[275,0.7621],[276,0.7533],[277,0.7373],[278,0.7235],[280,0.686],[281,0.6772],[285,0.6736],[286,0.6865],[287,0.6834],[288,0.6765],[289,0.6591],[290,0.655],[291,0.6548],[292,0.6523],[293,0.6514],[294,0.6664],[295,0.6778],[296,0.6794],[297,0.6787],[298,0.6578],[300,0.6478],[301,0.6333],[302,0.4843],[303,0.4629],[304,0.4048],[305,0.4008],[306,0.4061],[307,0.4048],[309,0.3903],[310,0.3881],[311,0.3731],[312,0.3714],[313,0.369],[314,0.3659],[316,0.3723],[317,0.3763],[318,0.3752],[319,0.375],[320,0.3712],[321,0.3695],[322,0.3689],[323,0.3745],[324,0.3772],[325,0.3767],[326,0.3744],[327,0.3751],[328,0.3703],[329,0.3684],[330,0.3617],[331,0.3604],[332,0.353],[334,0.3472],[335,0.3462],[336,0.3436],[337,0.3403],[338,0.3421],[339,0.3447],[340,0.353],[341,0.3555],[342,0.3542],[344,0.3439],[345,0.3438],[347,0.354],[348,0.3576],[349,0.3571],[350,0.3584],[351,0.3568],[352,0.3505],[353,0.3382],[354,0.3338],[355,0.3307],[356,0.3298],[358,0.3254],[360,0.325]]
       }
       CTRL_L_brow_raiseIn {
        type=MovieSceneScriptingFloatChannel
        default=0.2801
        thinned_from=492
        keys [[-131,0.3081],[-130,0.3159],[-129,0.3163],[-128,0.3019],[-127,0.2991],[-126,0.2957],[-125,0.2967],[-124,0.2901],[-123,0.2795],[-122,0.2771],[-121,0.2779],[-120,0.2838],[-119,0.2714],[-118,0.2687],[-117,0.2672],[-116,0.2582],[-115,0.2596],[-114,0.2598],[-113,0.2654],[-112,0.2648],[-111,0.2579],[-110,0.2446],[-109,0.236],[-108,0.2296],[-107,0.2258],[-106,0.2254],[-105,0.2239],[-103,0.2181],[-102,0.22],[-101,0.2327],[-100,0.2387],[-99,0.2342],[-98,0.226],[-97,0.2261],[-96,0.2247],[-95,0.2238],[-94,0.2214],[-93,0.2211],[-92,0.2236],[-91,0.2133],[-90,0.2085],[-89,0.2063],[-88,0.2157],[-87,0.2274],[-86,0.2316],[-85,0.2281],[-84,0.2198],[-83,0.2167],[-82,0.2217],[-81,0.2255],[-80,0.2208],[-79,0.22],[-78,0.2125],[-75,0.1852],[-74,0.1849],[-73,0.1949],[-72,0.1946],[-71,0.1893],[-70,0.1789],[-69,0.1802],[-68,0.1781],[-67,0.1847],[-66,0.1802],[-65,0.1767],[-64,0.1711],[-63,0.1678],[-62,0.1672],[-61,0.1703],[-60,0.173],[-59,0.1708],[-58,0.1606],[-57,0.1598],[-56,0.1572],[-55,0.1533],[-54,0.1531],[-53,0.1589],[-52,0.1627],[-51,0.1765],[-50,0.1751],[-49,0.1747],[-48,0.1733],[-47,0.1898],[-46,0.1894],[-45,0.1859],[-43,0.1833],[-42,0.1956],[-41,0.2015],[-40,0.1993],[-39,0.1988],[-37,0.2002],[-36,0.1893],[-35,0.172],[-34,0.1698],[-33,0.1621],[-32,0.1632],[-31,0.1626],[-30,0.1594],[-29,0.1593],[-28,0.1584],[-26,0.1664],[-25,0.1881],[-24,0.1979],[-23,0.2675],[-22,0.3867],[-21,0.5371],[-20,0.6544],[-19,0.7203],[-18,0.7431],[-17,0.7589],[-16,0.7834],[-15,0.8008],[-13,0.8018],[-12,0.7972],[-11,0.7896],[-10,0.7798],[-9,0.7902],[-8,0.7895],[-7,0.7793],[-6,0.7578],[-5,0.7501],[-3,0.7527],[-1,0.7515],[0,0.7356],[1,0.7232],[2,0.7078],[3,0.705],[4,0.7058],[5,0.6988],[6,0.673],[7,0.6614],[8,0.6581],[10,0.6457],[11,0.6319],[12,0.6209],[13,0.6196],[14,0.6317],[15,0.6338],[16,0.623],[17,0.6101],[18,0.5942],[20,0.5653],[21,0.5518],[22,0.5576],[23,0.5599],[24,0.57],[25,0.582],[27,0.5754],[28,0.5649],[29,0.543],[30,0.5256],[31,0.5236],[32,0.537],[33,0.5438],[34,0.5447],[35,0.5442],[36,0.5394],[37,0.5281],[39,0.5021],[40,0.5034],[41,0.5098],[42,0.51],[43,0.5054],[44,0.5138],[45,0.5172],[46,0.5934],[47,0.6118],[48,0.6544],[49,0.656],[50,0.7054],[51,0.711],[52,0.7058],[53,0.6544],[54,0.6325],[56,0.5257],[57,0.4755],[58,0.4426],[59,0.4198],[60,0.4165],[61,0.3992],[62,0.3575],[63,0.3542],[64,0.3014],[65,0.3014],[66,0.2851],[68,0.2788],[69,0.2689],[70,0.253],[71,0.243],[72,0.2364],[73,0.2181],[74,0.2052],[75,0.1902],[76,0.1789],[77,0.172],[78,0.1721],[79,0.1747],[80,0.1769],[81,0.1755],[82,0.1715],[83,0.1644],[84,0.1671],[85,0.1731],[86,0.185],[87,0.1853],[88,0.1851],[89,0.1777],[90,0.1742],[91,0.1662],[93,0.162],[94,0.1558],[95,0.1491],[96,0.1456],[97,0.1497],[98,0.1519],[99,0.1461],[100,0.134],[101,0.1318],[102,0.1323],[103,0.1267],[104,0.1289],[105,0.1292],[106,0.1483],[107,0.1574],[113,0.1573],[115,0.1579],[116,0.1586],[117,0.1606],[118,0.1648],[119,0.171],[120,0.1787],[122,0.192],[123,0.2002],[124,0.2097],[125,0.2229],[126,0.2356],[127,0.251],[128,0.2631],[129,0.2793],[130,0.297],[131,0.3221],[132,0.3548],[133,0.3953],[134,0.4692],[135,0.5584],[136,0.6625],[137,0.7511],[138,0.7864],[139,0.7491],[140,0.6809],[141,0.662],[142,0.6384],[143,0.6459],[144,0.6575],[146,0.6693],[147,0.6632],[148,0.6535],[149,0.6352],[150,0.6202],[151,0.6014],[153,0.5831],[154,0.5771],[155,0.5675],[157,0.5541],[159,0.5332],[161,0.5401],[162,0.5419],[163,0.5362],[164,0.5335],[165,0.5345],[166,0.5332],[167,0.5337],[168,0.5363],[169,0.5359],[171,0.5225],[172,0.5222],[174,0.5123],[175,0.5138],[176,0.5227],[177,0.5223],[178,0.5183],[179,0.5083],[180,0.5005],[181,0.5055],[182,0.5156],[183,0.5157],[184,0.4975],[185,0.4716],[186,0.4531],[187,0.4492],[189,0.4485],[190,0.4426],[191,0.4336],[192,0.414],[193,0.4318],[194,0.5817],[195,0.6499],[196,0.8049],[197,0.9141],[198,0.9823],[199,1],[201,0.9994],[202,0.9615],[203,0.939],[204,0.9136],[205,0.9234],[206,0.9076],[207,0.8985],[208,0.8826],[209,0.8559],[210,0.8204],[211,0.7817],[212,0.7729],[213,0.7557],[214,0.7518],[216,0.7237],[217,0.7053],[218,0.6887],[219,0.6697],[220,0.655],[222,0.6383],[223,0.6242],[224,0.6181],[225,0.6177],[226,0.6219],[228,0.6151],[229,0.6152],[230,0.6169],[231,0.6252],[232,0.6284],[233,0.6411],[234,0.6361],[235,0.6261],[236,0.605],[237,0.5912],[238,0.5794],[239,0.545],[240,0.4921],[241,0.4304],[242,0.4187],[243,0.4054],[244,0.3829],[245,0.3412],[246,0.2793],[247,0.2541],[248,0.2376],[249,0.2291],[250,0.2465],[251,0.2633],[252,0.2894],[253,0.2883],[254,0.2994],[255,0.3184],[256,0.3195],[257,0.2996],[258,0.2902],[259,0.2839],[260,0.2145],[261,0.1549],[262,0.103],[263,0.0702],[264,0.0652],[265,0.0523],[266,0.052],[267,0.0684],[268,0.1164],[269,0.18],[270,0.1899],[271,0.28],[272,0.3751],[273,0.4374],[274,0.4984],[275,0.5076],[276,0.497],[278,0.4609],[280,0.4162],[281,0.4056],[282,0.401],[283,0.3947],[284,0.3939],[285,0.3952],[286,0.4164],[287,0.413],[288,0.4054],[289,0.3882],[290,0.3845],[291,0.3818],[292,0.364],[293,0.3326],[294,0.3236],[295,0.3339],[296,0.346],[297,0.3479],[298,0.3512],[299,0.3537],[300,0.3472],[301,0.3336],[302,0.1946],[303,0.176],[304,0.1258],[305,0.1224],[306,0.1271],[307,0.1254],[309,0.1143],[310,0.1122],[311,0.0963],[312,0.094],[313,0.0893],[314,0.0852],[315,0.0856],[316,0.0879],[317,0.0909],[319,0.0881],[320,0.0853],[321,0.0841],[322,0.0835],[323,0.0855],[324,0.087],[325,0.0865],[326,0.0852],[327,0.0852],[328,0.0828],[329,0.081],[330,0.0774],[331,0.0763],[332,0.0728],[333,0.0709],[334,0.0695],[335,0.069],[336,0.068],[337,0.0665],[338,0.0672],[339,0.0677],[340,0.0707],[341,0.0717],[342,0.0718],[343,0.0693],[344,0.0671],[345,0.0669],[347,0.0697],[348,0.0702],[349,0.0704],[350,0.0713],[351,0.0714],[352,0.0694],[353,0.0647],[354,0.0625],[355,0.0608],[356,0.0604],[357,0.0595],[358,0.0592],[359,0.0584],[360,0.0584]]
       }
       CTRL_R_brow_raiseIn {
        type=MovieSceneScriptingFloatChannel
        default=0.43
        thinned_from=492
        keys [[-131,0.2543],[-130,0.2569],[-129,0.2555],[-128,0.2425],[-127,0.2429],[-126,0.2411],[-125,0.2387],[-124,0.2306],[-123,0.2217],[-122,0.2198],[-121,0.2215],[-120,0.2279],[-119,0.2192],[-118,0.2193],[-117,0.2185],[-116,0.214],[-115,0.2169],[-114,0.2153],[-113,0.2199],[-112,0.2211],[-111,0.2169],[-110,0.2048],[-109,0.1945],[-108,0.1851],[-107,0.1788],[-105,0.1771],[-104,0.1733],[-103,0.1713],[-102,0.1705],[-101,0.1844],[-100,0.1888],[-99,0.1847],[-98,0.1764],[-97,0.178],[-95,0.176],[-94,0.1687],[-93,0.1688],[-92,0.177],[-91,0.1707],[-90,0.1699],[-89,0.1686],[-88,0.1791],[-87,0.1913],[-86,0.1961],[-85,0.1918],[-84,0.1829],[-83,0.1799],[-81,0.1896],[-80,0.1864],[-79,0.1861],[-78,0.1828],[-77,0.1738],[-76,0.1629],[-75,0.1485],[-74,0.1482],[-73,0.1584],[-72,0.1586],[-71,0.1523],[-70,0.1419],[-69,0.1445],[-68,0.1381],[-67,0.1448],[-66,0.1403],[-65,0.1397],[-64,0.1347],[-63,0.1306],[-62,0.1295],[-60,0.1374],[-59,0.1354],[-58,0.123],[-57,0.1217],[-56,0.1182],[-55,0.1133],[-54,0.1109],[-53,0.115],[-52,0.1181],[-51,0.1355],[-50,0.1367],[-49,0.1366],[-48,0.1339],[-47,0.1473],[-46,0.1471],[-45,0.1456],[-44,0.1446],[-43,0.1444],[-42,0.1563],[-41,0.1626],[-40,0.1616],[-39,0.1609],[-38,0.1612],[-37,0.1623],[-36,0.1515],[-35,0.1345],[-34,0.1318],[-33,0.1257],[-32,0.1257],[-31,0.1248],[-30,0.1203],[-29,0.1196],[-28,0.1182],[-27,0.1231],[-26,0.1299],[-25,0.1517],[-24,0.1609],[-23,0.2177],[-22,0.3161],[-21,0.4506],[-20,0.5736],[-19,0.6362],[-18,0.6564],[-17,0.6593],[-16,0.6821],[-15,0.7023],[-13,0.7007],[-12,0.6882],[-11,0.6779],[-10,0.6641],[-9,0.6701],[-8,0.6676],[-7,0.6576],[-6,0.6351],[-5,0.6255],[-4,0.626],[-2,0.6192],[-1,0.619],[1,0.5997],[2,0.5848],[3,0.5797],[4,0.5786],[5,0.5757],[6,0.5555],[7,0.5448],[8,0.5423],[10,0.5342],[11,0.5216],[12,0.5124],[13,0.5127],[14,0.5225],[15,0.5209],[17,0.4994],[18,0.4899],[19,0.4772],[20,0.4689],[21,0.4572],[22,0.4655],[23,0.4666],[24,0.4731],[25,0.485],[26,0.4803],[27,0.4741],[28,0.4617],[29,0.4432],[30,0.4322],[31,0.4314],[32,0.4436],[33,0.4476],[34,0.449],[35,0.4486],[36,0.4444],[37,0.4332],[38,0.4198],[39,0.4052],[40,0.4067],[41,0.4152],[42,0.4175],[43,0.4118],[44,0.4187],[45,0.4224],[46,0.4926],[47,0.5155],[48,0.5704],[49,0.5839],[50,0.6417],[51,0.6589],[52,0.6617],[53,0.622],[54,0.6003],[57,0.4606],[58,0.4261],[59,0.4005],[60,0.3968],[61,0.383],[62,0.3482],[63,0.345],[64,0.2953],[65,0.2953],[66,0.2838],[67,0.2822],[68,0.2799],[69,0.2687],[70,0.2527],[71,0.2443],[72,0.2384],[73,0.2233],[74,0.2113],[75,0.1965],[76,0.1844],[77,0.1773],[78,0.1775],[79,0.1799],[80,0.1813],[81,0.1795],[82,0.1755],[83,0.1654],[84,0.1689],[85,0.1751],[86,0.191],[87,0.1894],[88,0.189],[89,0.1818],[90,0.1787],[91,0.1711],[92,0.1706],[93,0.1686],[94,0.1627],[95,0.1561],[96,0.1526],[97,0.1566],[98,0.158],[99,0.1511],[100,0.1367],[101,0.1321],[102,0.1319],[103,0.1286],[104,0.1317],[105,0.1319],[106,0.1514],[107,0.16],[114,0.1606],[115,0.161],[116,0.1617],[117,0.1636],[118,0.1675],[119,0.1733],[120,0.1804],[122,0.1924],[123,0.1998],[124,0.2085],[125,0.2203],[126,0.2316],[127,0.246],[128,0.2575],[129,0.2737],[130,0.2908],[131,0.3131],[132,0.3423],[133,0.3779],[134,0.4477],[135,0.5307],[136,0.6275],[137,0.7092],[138,0.7356],[139,0.6946],[140,0.6309],[141,0.6175],[142,0.6084],[144,0.6302],[145,0.6382],[146,0.6433],[147,0.6411],[148,0.6329],[149,0.6147],[150,0.5994],[151,0.5814],[152,0.5763],[153,0.5664],[154,0.5586],[155,0.5451],[156,0.5352],[157,0.5286],[158,0.5192],[159,0.5116],[161,0.5194],[162,0.5218],[163,0.5136],[164,0.511],[167,0.5148],[168,0.517],[169,0.5177],[171,0.5021],[172,0.4985],[173,0.4908],[174,0.485],[175,0.4863],[176,0.494],[177,0.4936],[178,0.4897],[179,0.4796],[180,0.4726],[181,0.4805],[182,0.4939],[183,0.4941],[184,0.4748],[185,0.4476],[186,0.4309],[187,0.4285],[188,0.4284],[189,0.43],[190,0.4227],[191,0.412],[192,0.3837],[193,0.3876],[194,0.5164],[195,0.5835],[196,0.7453],[197,0.8724],[198,0.9643],[199,0.997],[200,0.9999],[201,0.999],[202,0.9356],[203,0.891],[204,0.8494],[205,0.8715],[206,0.8604],[207,0.8529],[208,0.8383],[209,0.8121],[210,0.7786],[211,0.7413],[212,0.7336],[213,0.7178],[214,0.7157],[215,0.6997],[216,0.6865],[219,0.6348],[220,0.6228],[222,0.6078],[223,0.5933],[224,0.5897],[225,0.5897],[226,0.5943],[227,0.5918],[228,0.5908],[230,0.5956],[231,0.6054],[232,0.6089],[233,0.6208],[234,0.6174],[235,0.6114],[236,0.5888],[239,0.5031],[240,0.4704],[241,0.4237],[242,0.4142],[243,0.3935],[244,0.3763],[245,0.3361],[246,0.276],[247,0.2468],[248,0.2298],[249,0.2211],[250,0.239],[251,0.2552],[252,0.2794],[253,0.2777],[254,0.2907],[255,0.3106],[256,0.3128],[257,0.2882],[258,0.2709],[259,0.265],[260,0.1976],[261,0.1336],[262,0.0744],[263,0.0278],[264,0.0256],[265,0.0194],[266,0.0189],[267,0.0288],[268,0.0646],[269,0.1167],[270,0.1249],[271,0.2023],[272,0.3075],[273,0.3748],[274,0.4434],[275,0.4544],[276,0.4496],[277,0.4381],[278,0.4233],[280,0.3819],[281,0.3733],[282,0.3699],[283,0.3653],[284,0.3648],[285,0.3676],[286,0.3885],[287,0.3857],[288,0.3788],[289,0.3619],[291,0.3534],[292,0.3275],[293,0.28],[294,0.2614],[295,0.2758],[296,0.2937],[297,0.2966],[298,0.3049],[299,0.3086],[300,0.3046],[301,0.2925],[302,0.168],[303,0.1518],[304,0.1078],[305,0.1047],[306,0.1093],[307,0.1074],[308,0.102],[309,0.0956],[310,0.0931],[311,0.0739],[312,0.0709],[314,0.0608],[315,0.0615],[316,0.0632],[317,0.0668],[318,0.0658],[319,0.0658],[320,0.0653],[321,0.0653],[322,0.0646],[323,0.0652],[324,0.0672],[325,0.0686],[326,0.0693],[327,0.0715],[328,0.071],[329,0.0677],[330,0.0636],[331,0.0622],[333,0.0603],[335,0.0609],[336,0.0614],[337,0.0607],[338,0.0628],[339,0.0639],[340,0.0658],[342,0.0665],[343,0.0653],[344,0.0644],[345,0.0642],[347,0.0667],[349,0.0665],[350,0.0672],[351,0.0688],[352,0.0689],[353,0.0679],[354,0.0673],[355,0.0655],[356,0.0649],[357,0.0627],[358,0.0619],[360,0.0618]]
       }
       CTRL_R_brow_raiseOut {
        type=MovieSceneScriptingFloatChannel
        default=0
        thinned_from=492
        keys [[-131,0.4486],[-130,0.4601],[-128,0.45],[-127,0.4553],[-126,0.459],[-125,0.4551],[-124,0.4472],[-123,0.4419],[-122,0.441],[-121,0.4441],[-120,0.4507],[-119,0.4409],[-118,0.439],[-117,0.438],[-116,0.4316],[-115,0.4308],[-114,0.4311],[-113,0.4401],[-112,0.445],[-111,0.4388],[-109,0.4074],[-108,0.395],[-107,0.3854],[-106,0.384],[-105,0.3786],[-104,0.3762],[-103,0.3799],[-102,0.3889],[-101,0.406],[-100,0.4072],[-99,0.3962],[-98,0.3864],[-97,0.3859],[-96,0.3822],[-95,0.3806],[-94,0.3758],[-93,0.376],[-92,0.3841],[-91,0.3788],[-90,0.3811],[-89,0.3762],[-88,0.3799],[-87,0.3828],[-86,0.3789],[-85,0.363],[-84,0.3399],[-83,0.3293],[-82,0.327],[-81,0.3334],[-80,0.3296],[-79,0.3299],[-78,0.3324],[-77,0.3236],[-76,0.316],[-75,0.302],[-74,0.3041],[-73,0.315],[-72,0.3166],[-71,0.3138],[-70,0.308],[-69,0.3101],[-68,0.3059],[-67,0.3094],[-66,0.3074],[-64,0.2996],[-63,0.2976],[-62,0.298],[-61,0.3006],[-60,0.3005],[-59,0.2964],[-58,0.2859],[-57,0.2858],[-56,0.2847],[-55,0.2799],[-54,0.2774],[-53,0.2813],[-52,0.2841],[-51,0.2968],[-50,0.2985],[-49,0.2985],[-48,0.3],[-47,0.3143],[-46,0.3138],[-44,0.3103],[-43,0.3067],[-42,0.314],[-41,0.318],[-39,0.3129],[-38,0.3125],[-37,0.3105],[-36,0.3028],[-35,0.2908],[-33,0.2849],[-32,0.2892],[-31,0.2896],[-30,0.2906],[-29,0.2901],[-28,0.2888],[-27,0.2895],[-26,0.291],[-25,0.3021],[-24,0.3077],[-23,0.3784],[-22,0.4746],[-21,0.6086],[-20,0.6994],[-19,0.6965],[-18,0.6213],[-17,0.5299],[-16,0.4879],[-15,0.4729],[-14,0.4524],[-13,0.4406],[-12,0.4062],[-10,0.3425],[-9,0.3167],[-7,0.2768],[-6,0.2549],[-5,0.247],[-4,0.2434],[-3,0.2378],[-2,0.2285],[-1,0.2298],[0,0.222],[1,0.2151],[2,0.2008],[3,0.2009],[4,0.2049],[5,0.2059],[6,0.1992],[8,0.2028],[9,0.2089],[10,0.2124],[11,0.2085],[12,0.203],[13,0.2025],[14,0.2091],[15,0.2118],[16,0.2099],[18,0.2092],[19,0.2053],[20,0.2061],[21,0.2029],[22,0.2116],[23,0.2112],[24,0.2116],[25,0.2182],[26,0.2194],[27,0.2219],[29,0.2212],[30,0.2235],[31,0.2284],[32,0.2393],[33,0.2397],[35,0.2398],[37,0.2438],[38,0.2343],[39,0.221],[40,0.2186],[41,0.2282],[42,0.2335],[43,0.2319],[44,0.2363],[45,0.2388],[46,0.2985],[47,0.3533],[48,0.4661],[49,0.5231],[50,0.6047],[51,0.6563],[52,0.6837],[53,0.6793],[54,0.6686],[55,0.6548],[56,0.6446],[57,0.6296],[58,0.6183],[59,0.6167],[60,0.6172],[61,0.6196],[62,0.6043],[63,0.6021],[64,0.5671],[65,0.5671],[66,0.5587],[67,0.5575],[68,0.5548],[69,0.5457],[70,0.5317],[71,0.5246],[72,0.5145],[73,0.4925],[75,0.4624],[76,0.4464],[77,0.4395],[78,0.4394],[79,0.438],[80,0.4403],[81,0.4364],[82,0.4308],[83,0.4166],[84,0.4191],[85,0.4253],[86,0.4399],[87,0.4337],[88,0.43],[89,0.4234],[90,0.423],[91,0.4164],[92,0.4163],[93,0.4153],[94,0.4111],[95,0.4038],[96,0.4007],[98,0.4074],[99,0.4016],[100,0.3892],[101,0.3859],[102,0.386],[103,0.3819],[104,0.3855],[105,0.3857],[106,0.4022],[107,0.412],[115,0.4131],[116,0.4139],[117,0.4156],[118,0.4191],[119,0.4239],[122,0.4423],[123,0.4497],[124,0.4583],[126,0.4795],[128,0.5069],[129,0.525],[130,0.5443],[131,0.568],[132,0.5955],[133,0.6272],[134,0.6893],[135,0.7717],[136,0.8711],[137,0.9481],[138,0.9727],[139,0.9221],[140,0.8633],[141,0.8474],[142,0.8408],[143,0.8584],[144,0.8794],[146,0.9039],[147,0.9059],[148,0.9061],[149,0.8898],[150,0.8782],[151,0.8624],[154,0.8512],[155,0.8422],[156,0.8295],[158,0.8166],[159,0.8123],[161,0.8094],[162,0.8043],[163,0.7967],[164,0.7914],[165,0.7941],[166,0.7896],[167,0.7905],[168,0.7889],[169,0.7895],[170,0.7844],[172,0.7698],[173,0.7586],[174,0.7558],[176,0.7655],[177,0.7634],[178,0.7529],[179,0.7359],[180,0.7258],[181,0.7267],[182,0.7327],[183,0.7279],[184,0.7106],[185,0.6848],[186,0.6687],[187,0.6634],[188,0.6628],[189,0.6643],[190,0.6495],[191,0.6302],[192,0.5776],[193,0.5628],[194,0.6849],[195,0.7484],[196,0.8757],[197,0.951],[198,0.9908],[199,1],[208,1],[209,0.9988],[210,0.984],[211,0.9579],[212,0.9499],[213,0.9345],[214,0.9358],[216,0.9126],[218,0.8774],[219,0.8634],[221,0.8445],[222,0.8364],[223,0.8238],[225,0.8142],[226,0.8137],[228,0.806],[230,0.7942],[231,0.7957],[232,0.7926],[233,0.8061],[234,0.8125],[235,0.8388],[236,0.8539],[237,0.8333],[238,0.7889],[239,0.7541],[240,0.7314],[241,0.694],[242,0.6857],[243,0.6482],[244,0.6157],[245,0.5501],[246,0.4657],[247,0.4304],[248,0.4131],[249,0.4119],[250,0.4289],[251,0.4507],[252,0.4703],[253,0.4744],[254,0.4852],[255,0.5071],[256,0.5099],[257,0.4951],[258,0.4972],[259,0.494],[260,0.4464],[261,0.3688],[262,0.2742],[263,0.1579],[264,0.1508],[265,0.1358],[266,0.1412],[267,0.1655],[268,0.2333],[269,0.3192],[270,0.3323],[271,0.4185],[272,0.4919],[273,0.5425],[274,0.6065],[275,0.6218],[276,0.6252],[277,0.6229],[278,0.6149],[279,0.5984],[280,0.583],[281,0.5774],[282,0.5774],[285,0.5804],[286,0.5939],[287,0.5923],[288,0.5866],[289,0.569],[290,0.5647],[291,0.5622],[292,0.5445],[293,0.5132],[294,0.5118],[295,0.5356],[296,0.5508],[297,0.5525],[298,0.5458],[299,0.548],[300,0.555],[301,0.5439],[302,0.4247],[303,0.4085],[304,0.3644],[305,0.3622],[306,0.3685],[307,0.3672],[308,0.3605],[309,0.3522],[310,0.3492],[311,0.324],[312,0.3206],[313,0.3164],[314,0.3096],[315,0.3135],[316,0.3163],[317,0.3231],[318,0.3231],[319,0.3259],[321,0.3273],[322,0.3265],[323,0.3289],[324,0.3339],[325,0.338],[326,0.3408],[327,0.3458],[328,0.3449],[329,0.338],[330,0.3294],[331,0.3267],[332,0.325],[333,0.3243],[334,0.326],[336,0.3274],[337,0.3258],[338,0.3316],[339,0.336],[340,0.3423],[341,0.3438],[342,0.3432],[343,0.3405],[344,0.3384],[345,0.3385],[347,0.3476],[348,0.3495],[349,0.3486],[350,0.3495],[351,0.3512],[352,0.3493],[353,0.3446],[354,0.3434],[355,0.3395],[356,0.3379],[357,0.333],[358,0.3296],[359,0.3309],[360,0.3309]]
       }
       CTRL_lookAtSwitch {type=MovieSceneScriptingFloatChannel, default=0, const [0,492]}
       CTRL_C_teethD.X {type=MovieSceneScriptingFloatChannel, default=0, const [0,492]}
       CTRL_C_teethD.Y {type=MovieSceneScriptingFloatChannel, default=0, const [0,492]}
       CTRL_C_teethU.X {type=MovieSceneScriptingFloatChannel, default=0, const [0,492]}
       CTRL_C_teethU.Y {type=MovieSceneScriptingFloatChannel, default=0, const [0,492]}
       CTRL_C_teeth_fwdBackU {type=MovieSceneScriptingFloatChannel, default=0, const [0,492]}
       CTRL_L_mouth_corner.X {type=MovieSceneScriptingFloatChannel, default=0, const [0,492]}
       CTRL_L_mouth_corner.Y {
        type=MovieSceneScriptingFloatChannel
        default=0
        thinned_from=492
        keys [[-131,0.0323],[-130,0.03],[-129,0.0295],[-128,0.03],[-127,0.03],[-126,0.0298],[-125,0.0305],[-123,0.0329],[-122,0.0332],[-120,0.0332],[-119,0.0335],[-117,0.0335],[-116,0.0328],[-115,0.0318],[-113,0.029],[-111,0.0266],[-110,0.0258],[-108,0.0251],[-107,0.0242],[-105,0.0237],[-104,0.0232],[-103,0.0232],[-102,0.0224],[-101,0.0218],[-100,0.0215],[-99,0.0223],[-98,0.0223],[-97,0.0212],[-95,0.0215],[-94,0.0212],[-93,0.0212],[-92,0.0217],[-91,0.0247],[-90,0.0282],[-88,0.0386],[-87,0.0432],[-85,0.049],[-84,0.0531],[-83,0.0581],[-82,0.064],[-81,0.0674],[-80,0.0693],[-79,0.0694],[-78,0.0696],[-77,0.0685],[-76,0.0681],[-75,0.07],[-74,0.0742],[-73,0.0791],[-72,0.0821],[-71,0.084],[-69,0.0866],[-67,0.0902],[-65,0.0931],[-64,0.093],[-63,0.0934],[-61,0.0982],[-60,0.0978],[-59,0.0882],[-58,0.0664],[-57,0.062],[-56,0.0521],[-55,0.0458],[-54,0.0417],[-53,0.0386],[-52,0.037],[-51,0.0337],[-50,0.0326],[-49,0.0326],[-48,0.0306],[-47,0.0309],[-46,0.0318],[-45,0.0367],[-44,0.0378],[-43,0.0415],[-41,0.052],[-39,0.0674],[-38,0.0696],[-37,0.0794],[-36,0.0831],[-35,0.0882],[-34,0.092],[-33,0.0935],[-32,0.087],[-31,0.0829],[-30,0.0575],[-29,0.0537],[-28,0.0456],[-27,0.0418],[-26,0.04],[-25,0.0388],[-24,0.0384],[-23,0.0372],[-22,0.0361],[-21,0.0359],[-20,0.0375],[-19,0.0384],[-18,0.04],[-17,0.0394],[-16,0.0383],[-15,0.0368],[-14,0.0362],[-13,0.0365],[-12,0.0371],[-11,0.0365],[-10,0.0365],[-9,0.0361],[-8,0.0361],[-5,0.0351],[-4,0.0342],[-1,0.0322],[1,0.0316],[2,0.0316],[3,0.0309],[4,0.0309],[5,0.0316],[6,0.0332],[7,0.0335],[10,0.0333],[12,0.0324],[13,0.0324],[15,0.0333],[16,0.0333],[17,0.0337],[18,0.0336],[19,0.0342],[20,0.0342],[21,0.0352],[25,0.0353],[27,0.0361],[29,0.0361],[31,0.0358],[32,0.0353],[33,0.0352],[34,0.0358],[35,0.0357],[36,0.0351],[37,0.0347],[38,0.0348],[39,0.0355],[40,0.0352],[42,0.0333],[43,0.0332],[44,0.0335],[45,0.0335],[46,0.0328],[47,0.0316],[48,0.0307],[49,0.0309],[51,0.0286],[52,0.0283],[53,0.0329],[55,0.0333],[56,0.0326],[58,0.0321],[59,0.0316],[61,0.0317],[62,0.0342],[63,0.0342],[64,0.0348],[66,0.0348],[67,0.0345],[68,0.0348],[69,0.0348],[70,0.0355],[71,0.0352],[72,0.0352],[73,0.0354],[74,0.0368],[75,0.0387],[76,0.0402],[77,0.0397],[78,0.0396],[79,0.0381],[80,0.0368],[81,0.0368],[82,0.0362],[83,0.0365],[84,0.0361],[85,0.0364],[87,0.0338],[88,0.0319],[89,0.0319],[90,0.0313],[91,0.0326],[92,0.0335],[93,0.0333],[94,0.0327],[95,0.0324],[96,0.0324],[98,0.0321],[99,0.0322],[100,0.0331],[101,0.0333],[102,0.0332],[103,0.0324],[105,0.0322],[106,0.032],[107,0.0322],[108,0.0319],[109,0.0326],[110,0.0335],[111,0.0352],[112,0.0355],[113,0.0355],[114,0.0349],[116,0.0361],[117,0.0379],[118,0.0403],[120,0.0432],[121,0.0417],[122,0.0413],[123,0.0414],[124,0.0423],[126,0.041],[127,0.0394],[128,0.0394],[130,0.0371],[133,0.0348],[134,0.0352],[135,0.0368],[136,0.0421],[137,0.0433],[138,0.0401],[139,0.0361],[140,0.0368],[141,0.0385],[142,0.0415],[146,0.0401],[147,0.041],[149,0.0416],[150,0.0407],[151,0.04],[153,0.0394],[154,0.0394],[156,0.0387],[157,0.0381],[158,0.0371],[159,0.0368],[162,0.0367],[163,0.0365],[164,0.0374],[165,0.0371],[166,0.0371],[167,0.0368],[168,0.0386],[169,0.0409],[171,0.0416],[172,0.0422],[175,0.0452],[177,0.0451],[179,0.0436],[180,0.0436],[182,0.0459],[183,0.0465],[184,0.0456],[185,0.0443],[186,0.044],[187,0.0451],[189,0.0456],[190,0.0445],[191,0.044],[192,0.0444],[194,0.0459],[195,0.0461],[196,0.0441],[197,0.0413],[198,0.0382],[199,0.0355],[200,0.0349],[201,0.0339],[202,0.0279],[204,0.0309],[205,0.0379],[206,0.0367],[207,0.0351],[208,0.0342],[209,0.0339],[211,0.0358],[212,0.036],[213,0.0365],[214,0.0358],[215,0.0361],[216,0.0356],[218,0.0365],[219,0.0365],[221,0.0358],[222,0.0361],[223,0.0361],[224,0.0354],[225,0.0345],[226,0.0342],[228,0.0341],[229,0.0332],[231,0.0319],[232,0.0322],[233,0.0322],[234,0.0316],[235,0.0302],[236,0.0274],[237,0.0243],[238,0.0219],[239,0.0242],[240,0.0295],[241,0.0336],[242,0.0342],[243,0.0295],[244,0.0271],[245,0.0261],[247,0.0327],[249,0.0418],[250,0.0456],[251,0.0462],[252,0.0473],[253,0.0491],[254,0.0518],[255,0.0524],[256,0.0523],[257,0.0511],[258,0.0469],[259,0.0467],[261,0.0411],[263,0.0404],[264,0.0409],[265,0.0428],[266,0.0432],[267,0.0434],[268,0.0431],[269,0.044],[270,0.0443],[273,0.0403],[274,0.0386],[275,0.038],[276,0.038],[277,0.0387],[278,0.0391],[279,0.0386],[280,0.0384],[281,0.0389],[282,0.0391],[283,0.0388],[284,0.0387],[285,0.0381],[286,0.0372],[288,0.0371],[290,0.0363],[291,0.0353],[292,0.0347],[293,0.0353],[294,0.0364],[295,0.0368],[296,0.0359],[297,0.0358],[298,0.0349],[299,0.0343],[300,0.0332],[301,0.0331],[302,0.0326],[304,0.0327],[305,0.0329],[306,0.0335],[307,0.0343],[308,0.0349],[310,0.0352],[311,0.0363],[313,0.0365],[314,0.0367],[318,0.0345],[319,0.0345],[320,0.0348],[321,0.0358],[322,0.0364],[323,0.0368],[324,0.0368],[325,0.0365],[326,0.0365],[328,0.0371],[331,0.0371],[332,0.0374],[335,0.0365],[336,0.0371],[337,0.0374],[338,0.0381],[339,0.0384],[340,0.039],[341,0.0381],[342,0.0378],[343,0.0378],[344,0.0384],[345,0.0387],[349,0.0388],[353,0.04],[355,0.0397],[356,0.0397],[358,0.0404],[360,0.0403]]
       }
       CTRL_R_mouth_corner.X {type=MovieSceneScriptingFloatChannel, default=0, const [0,492]}
       CTRL_R_mouth_corner.Y {
        type=MovieSceneScriptingFloatChannel
        default=0
        thinned_from=492
        keys [[-131,-0.0067],[-130,-0.0118],[-129,-0.0124],[-128,-0.0109],[-127,-0.0099],[-126,-0.0096],[-125,-0.0075],[-124,-0.0042],[-123,-0.0031],[-122,-0.0029],[-121,-0.003],[-120,-0.0016],[-119,-0.0009],[-118,-0.0021],[-117,-0.0021],[-116,-0.0018],[-115,-0.0021],[-113,-0.0055],[-110,-0.0092],[-109,-0.0101],[-107,-0.0145],[-106,-0.0149],[-105,-0.0161],[-104,-0.0171],[-103,-0.017],[-101,-0.0192],[-99,-0.0228],[-98,-0.0243],[-97,-0.0264],[-96,-0.0275],[-95,-0.0277],[-94,-0.0295],[-93,-0.0295],[-92,-0.028],[-91,-0.0259],[-90,-0.0217],[-89,-0.0135],[-88,-0.0026],[-87,0.0093],[-86,0.0187],[-84,0.0361],[-83,0.0441],[-82,0.0509],[-81,0.055],[-80,0.0544],[-79,0.054],[-78,0.0505],[-77,0.0442],[-76,0.0402],[-75,0.0365],[-74,0.0355],[-73,0.0355],[-72,0.0359],[-71,0.0374],[-70,0.0381],[-69,0.0394],[-68,0.0381],[-67,0.0411],[-66,0.0429],[-65,0.0467],[-64,0.0479],[-63,0.0506],[-62,0.0538],[-61,0.0573],[-60,0.0596],[-59,0.0561],[-58,0.0436],[-57,0.041],[-56,0.035],[-55,0.0313],[-54,0.0293],[-53,0.0283],[-52,0.0279],[-51,0.026],[-50,0.0249],[-49,0.0247],[-47,0.0293],[-46,0.0299],[-45,0.0341],[-44,0.0349],[-43,0.0374],[-42,0.0395],[-41,0.0414],[-39,0.0492],[-38,0.0501],[-36,0.0534],[-35,0.056],[-34,0.0589],[-33,0.0608],[-31,0.0572],[-30,0.045],[-29,0.043],[-28,0.0385],[-27,0.0364],[-26,0.0365],[-25,0.0368],[-24,0.0368],[-23,0.0355],[-22,0.0321],[-21,0.0284],[-20,0.0279],[-19,0.0254],[-18,0.0233],[-17,0.0189],[-16,0.0182],[-15,0.0179],[-14,0.0179],[-13,0.0176],[-12,0.0176],[-11,0.0182],[-10,0.0182],[-9,0.0174],[-8,0.0159],[-7,0.015],[-6,0.0153],[-5,0.0166],[-3,0.0172],[-1,0.0166],[0,0.0156],[1,0.0151],[3,0.0153],[4,0.015],[5,0.0137],[6,0.0137],[7,0.0152],[9,0.0155],[10,0.0148],[11,0.0138],[14,0.0127],[15,0.0133],[16,0.0137],[17,0.0137],[18,0.0122],[20,0.0117],[21,0.0137],[22,0.0146],[23,0.0149],[24,0.0157],[26,0.0166],[27,0.0168],[29,0.0178],[30,0.0177],[31,0.0173],[33,0.0176],[34,0.0173],[36,0.0173],[37,0.0177],[38,0.0178],[39,0.0173],[40,0.016],[41,0.0151],[42,0.0145],[43,0.0148],[44,0.0157],[45,0.0159],[46,0.0189],[47,0.0195],[48,0.0192],[50,0.0154],[51,0.0128],[52,0.0114],[53,0.0137],[54,0.0149],[55,0.0167],[56,0.0175],[57,0.0181],[58,0.0171],[59,0.015],[60,0.0147],[61,0.0135],[62,0.0156],[63,0.0157],[64,0.0169],[65,0.0169],[68,0.0179],[69,0.0179],[70,0.0183],[73,0.0211],[74,0.022],[75,0.0232],[76,0.025],[77,0.026],[78,0.026],[79,0.0257],[80,0.0251],[81,0.0251],[82,0.0254],[83,0.0254],[85,0.0247],[86,0.0251],[87,0.0244],[88,0.0244],[89,0.0242],[90,0.0247],[92,0.0254],[93,0.0253],[94,0.0248],[95,0.0241],[97,0.0234],[99,0.0217],[100,0.0217],[101,0.0223],[102,0.0224],[103,0.0216],[105,0.0215],[107,0.0221],[108,0.0222],[109,0.0228],[113,0.0241],[114,0.0241],[115,0.0244],[116,0.0244],[118,0.0257],[119,0.026],[120,0.0267],[121,0.0271],[122,0.028],[123,0.0276],[124,0.027],[125,0.0267],[126,0.0267],[128,0.0254],[129,0.0244],[131,0.0217],[132,0.0212],[133,0.0212],[135,0.0199],[136,0.0215],[137,0.0218],[138,0.0203],[139,0.0169],[140,0.0151],[141,0.0152],[142,0.0172],[143,0.0202],[144,0.0227],[145,0.0228],[146,0.0219],[147,0.0228],[148,0.0241],[149,0.0234],[151,0.0208],[152,0.0208],[153,0.0205],[154,0.0208],[155,0.0215],[156,0.0218],[157,0.0212],[158,0.0195],[159,0.0189],[160,0.0186],[161,0.0198],[162,0.0197],[163,0.0205],[164,0.021],[165,0.0225],[166,0.0229],[168,0.0273],[169,0.0313],[170,0.0319],[171,0.0312],[172,0.0302],[173,0.0296],[174,0.0279],[175,0.0257],[176,0.0229],[177,0.0225],[178,0.0211],[179,0.0193],[180,0.0182],[181,0.0162],[182,0.0156],[183,0.0154],[186,0.0141],[188,0.0143],[189,0.014],[191,0.0124],[193,0.009],[194,0.0054],[195,0.0052],[196,0.0011],[197,-0.0027],[198,-0.005],[199,-0.0028],[200,-0.0015],[201,0.004],[202,0.0025],[203,0.0006],[204,-0.0035],[205,-0.0001],[206,-0.0001],[207,-0.001],[209,-0.002],[210,-0.0003],[211,0.0017],[213,0.0029],[214,0.0023],[217,0.0029],[220,0.0039],[221,0.0045],[222,0.0049],[223,0.0046],[224,0.0032],[225,0.0023],[226,0.0026],[227,0.0033],[228,0.0045],[229,0.0042],[230,0.0052],[231,0.0059],[232,0.0068],[233,0.0057],[235,0.0029],[236,0.0022],[238,-0.0056],[239,-0.0067],[240,-0.0049],[241,-0.004],[242,-0.0039],[243,-0.007],[245,-0.0085],[246,-0.0109],[247,-0.0153],[248,-0.0182],[249,-0.0162],[250,-0.0114],[251,-0.0076],[252,-0.0015],[253,0.0065],[254,0.0152],[255,0.0201],[256,0.0205],[257,0.02],[258,0.0132],[259,0.0125],[260,0.0061],[261,0.0032],[262,0.0032],[264,0.0059],[265,0.01],[266,0.0106],[267,0.0087],[268,0.0041],[269,0.0006],[270,0.0002],[271,-0.0027],[272,-0.0005],[273,-0.001],[275,-0.0001],[276,0.0012],[277,0.0022],[278,0.0025],[280,0],[281,0.0009],[282,0.0016],[284,0.0016],[285,0],[287,-0.0019],[288,-0.0032],[289,-0.006],[290,-0.0075],[291,-0.0095],[292,-0.011],[293,-0.0128],[294,-0.0133],[295,-0.0129],[296,-0.012],[297,-0.0118],[298,-0.0088],[299,-0.0043],[300,0.0038],[301,0.0045],[302,0.0104],[303,0.0116],[304,0.0151],[305,0.0169],[306,0.0183],[307,0.0195],[308,0.0212],[309,0.0223],[310,0.0225],[311,0.0233],[312,0.0235],[313,0.0244],[314,0.0257],[315,0.0257],[316,0.0251],[317,0.0251],[319,0.0257],[320,0.0257],[322,0.027],[323,0.027],[324,0.0273],[326,0.0273],[327,0.0267],[328,0.0264],[331,0.0264],[332,0.0267],[333,0.0267],[334,0.0274],[335,0.0277],[336,0.0283],[338,0.029],[340,0.029],[341,0.0287],[342,0.0286],[344,0.028],[345,0.0283],[346,0.0283],[347,0.0286],[348,0.028],[349,0.0283],[350,0.0283],[351,0.029],[352,0.0287],[354,0.0295],[356,0.0296],[357,0.0293],[358,0.0293],[359,0.029],[360,0.029]]
       }
       CTRL_C_teeth_fwdBackD {type=MovieSceneScriptingFloatChannel, default=0, const [0,492]}
       CTRL_R_mouth_cornerSharpnessD {
        type=MovieSceneScriptingFloatChannel
        default=0
        thinned_from=492
        keys [[-131,0.0756],[-130,0.0743],[-128,0.0752],[-127,0.0774],[-126,0.0806],[-125,0.0847],[-124,0.0876],[-123,0.089],[-122,0.0894],[-121,0.0907],[-120,0.0881],[-119,0.086],[-118,0.0845],[-117,0.0841],[-116,0.0809],[-115,0.0782],[-114,0.0783],[-113,0.0782],[-112,0.0763],[-111,0.0734],[-110,0.072],[-109,0.0703],[-108,0.0646],[-107,0.0597],[-106,0.059],[-105,0.0572],[-104,0.0579],[-103,0.0591],[-102,0.059],[-101,0.0571],[-100,0.0548],[-99,0.0556],[-98,0.0527],[-97,0.052],[-96,0.053],[-95,0.0533],[-94,0.0552],[-93,0.0552],[-92,0.0557],[-91,0.0566],[-90,0.0599],[-89,0.0712],[-88,0.0855],[-87,0.0984],[-86,0.1023],[-85,0.1027],[-84,0.1037],[-83,0.1025],[-82,0.0966],[-81,0.0876],[-80,0.0805],[-79,0.0796],[-78,0.0707],[-77,0.0549],[-75,0.0411],[-74,0.0322],[-73,0.0317],[-72,0.0316],[-71,0.0298],[-69,0.0269],[-68,0.0283],[-67,0.0288],[-66,0.0288],[-65,0.0303],[-64,0.03],[-62,0.0262],[-61,0.0266],[-60,0.0245],[-59,0.0163],[-58,-0.0017],[-57,-0.0038],[-56,-0.0078],[-55,-0.0096],[-54,-0.0108],[-53,-0.0115],[-52,-0.0119],[-51,-0.0098],[-50,-0.0086],[-49,-0.0085],[-48,-0.0018],[-47,0.0014],[-46,0.0024],[-45,0.0087],[-44,0.0102],[-43,0.0156],[-42,0.0214],[-41,0.0264],[-40,0.0344],[-39,0.0401],[-38,0.041],[-37,0.0396],[-36,0.0349],[-35,0.0328],[-34,0.0337],[-33,0.029],[-32,0.0226],[-31,0.0201],[-30,0.0055],[-29,0.0045],[-27,0],[-26,-0.0015],[-25,-0.0033],[-24,-0.004],[-22,-0.0112],[-21,-0.0098],[-20,-0.0055],[-19,-0.0032],[-18,-0.0017],[-15,0.0005],[-14,0.001],[-13,0.001],[-12,0.0015],[-11,0.0015],[-10,0.0005],[-9,-0.0026],[-8,-0.004],[-7,-0.003],[-6,-0.0032],[-5,-0.0039],[-4,-0.0034],[-3,-0.0035],[-2,-0.0049],[-1,-0.0051],[0,-0.0049],[1,-0.0058],[3,-0.0054],[4,-0.0059],[5,-0.0054],[6,-0.0034],[7,-0.0025],[9,-0.0024],[10,-0.0028],[11,-0.0029],[12,-0.0036],[13,-0.0039],[14,-0.0032],[15,-0.002],[16,-0.0016],[17,-0.0017],[19,-0.0029],[20,-0.0013],[21,-0.0003],[22,-0.0009],[23,-0.0012],[24,-0.0019],[25,-0.0019],[26,-0.0017],[27,-0.0012],[28,-0.0011],[30,-0.002],[31,-0.0022],[33,-0.001],[35,-0.001],[36,-0.0013],[37,-0.0014],[39,0.0002],[40,-0.0012],[41,-0.0019],[42,-0.0015],[43,-0.0017],[44,-0.0024],[45,-0.0027],[46,-0.0041],[47,-0.0041],[50,-0.0022],[52,-0.0015],[53,-0.0015],[54,-0.001],[55,0.0008],[56,0.0033],[57,0.0051],[59,0.0076],[60,0.0078],[62,0.0088],[63,0.0088],[64,0.0093],[65,0.0093],[67,0.0074],[68,0.0078],[69,0.0069],[70,0.0054],[71,0.0054],[72,0.0049],[73,0.0052],[75,0.0063],[76,0.0062],[77,0.0039],[78,0.0038],[79,0.0024],[80,0.0029],[81,0.0039],[82,0.0034],[83,0.002],[84,0.0024],[85,0.0024],[86,0.0015],[87,0.001],[88,0.0015],[89,0.0014],[90,0.001],[91,0.0024],[92,0.0015],[94,0.0012],[95,0.0021],[96,0.0024],[97,0.0024],[98,0.0027],[99,0.0024],[100,0.002],[101,0.0026],[102,0.0029],[103,0.0016],[104,0.0019],[105,0.002],[106,0.0029],[107,0.0029],[108,0.0035],[110,0.0053],[111,0.0068],[113,0.0078],[114,0.0088],[115,0.0088],[116,0.0083],[117,0.0089],[118,0.0107],[119,0.0108],[120,0.0112],[121,0.0127],[123,0.0118],[124,0.0122],[126,0.0093],[127,0.0097],[128,0.0097],[129,0.0083],[130,0.0049],[131,0.0041],[132,0.0053],[133,0.0059],[134,0.0074],[136,0.0112],[137,0.0112],[138,0.0054],[139,-0.0015],[140,-0.0059],[141,-0.0068],[142,-0.0061],[143,-0.0031],[144,0.0018],[145,0.0034],[146,0.0025],[147,0.0029],[148,0.0024],[150,0.0054],[151,0.0033],[152,0.002],[153,-0.0002],[154,-0.0002],[155,0.0019],[156,0.0024],[157,0.0015],[159,-0.0012],[160,-0.0013],[161,0.0004],[162,0],[163,0],[164,0.0013],[165,0.001],[166,-0.0002],[167,0.0002],[168,0.0048],[169,0.0106],[170,0.0114],[171,0.012],[172,0.0137],[173,0.015],[175,0.0185],[177,0.0177],[178,0.0165],[179,0.0159],[180,0.0156],[181,0.0142],[182,0.0142],[184,0.0133],[185,0.0132],[186,0.0136],[187,0.0124],[188,0.0122],[189,0.0117],[190,0.0122],[191,0.0122],[192,0.0137],[193,0.0164],[194,0.0195],[195,0.0202],[196,0.023],[197,0.0256],[198,0.027],[199,0.028],[200,0.0283],[201,0.0267],[202,0.0286],[203,0.0257],[204,0.0183],[205,0.0171],[206,0.017],[208,0.0161],[209,0.0155],[210,0.0137],[211,0.0132],[212,0.0132],[214,0.0127],[215,0.0127],[216,0.0123],[217,0.0122],[218,0.0127],[219,0.0113],[220,0.0108],[221,0.0122],[222,0.0122],[224,0.0103],[225,0.0103],[226,0.0112],[227,0.0112],[228,0.0118],[229,0.0127],[230,0.0127],[232,0.0137],[233,0.0151],[234,0.0151],[235,0.0138],[236,0.019],[237,0.0191],[238,0.0159],[239,0.0144],[240,0.0124],[241,0.0117],[242,0.0117],[243,0.0122],[244,0.0122],[245,0.0128],[246,0.0199],[247,0.041],[248,0.0733],[249,0.095],[250,0.0896],[251,0.0813],[252,0.0711],[253,0.0655],[255,0.0664],[256,0.0637],[257,0.0678],[258,0.0733],[259,0.0738],[260,0.0778],[261,0.0766],[262,0.0731],[263,0.0659],[264,0.0637],[265,0.06],[266,0.059],[268,0.0503],[269,0.0525],[270,0.0532],[271,0.0572],[272,0.0576],[273,0.0588],[274,0.0628],[275,0.0655],[276,0.0673],[277,0.0687],[278,0.0684],[279,0.0662],[280,0.0659],[281,0.0691],[282,0.0695],[283,0.0676],[284,0.0674],[285,0.0675],[286,0.0686],[287,0.0657],[288,0.0607],[289,0.0548],[291,0.0539],[292,0.0528],[293,0.0532],[294,0.054],[295,0.0534],[296,0.0524],[297,0.0518],[298,0.0411],[299,0.03],[300,0.0191],[301,0.0186],[302,0.0151],[303,0.0147],[304,0.0133],[305,0.0126],[306,0.012],[307,0.012],[308,0.0122],[311,0.0123],[312,0.0122],[313,0.0117],[315,0.0117],[316,0.0112],[319,0.0112],[320,0.0122],[321,0.0127],[323,0.0127],[324,0.0122],[325,0.0122],[326,0.0127],[327,0.0127],[328,0.0122],[329,0.0123],[330,0.0132],[331,0.0132],[332,0.0127],[334,0.0137],[335,0.0137],[336,0.0127],[337,0.0122],[338,0.0132],[339,0.0132],[341,0.0142],[342,0.0137],[343,0.0142],[344,0.0142],[345,0.0146],[346,0.0146],[347,0.0142],[348,0.0146],[349,0.0142],[351,0.0142],[353,0.0132],[354,0.0132],[355,0.0128],[357,0.0127],[358,0.0122],[359,0.0112],[360,0.0112]]
       }
       CTRL_R_mouth_cornerSharpnessU {
        type=MovieSceneScriptingFloatChannel
        default=0
        thinned_from=492
        keys [[-131,0.1187],[-129,0.1197],[-128,0.1181],[-126,0.1215],[-125,0.1221],[-124,0.1222],[-123,0.1217],[-122,0.1216],[-121,0.1229],[-120,0.1199],[-119,0.1164],[-118,0.1152],[-117,0.1147],[-116,0.1105],[-115,0.1078],[-114,0.1093],[-113,0.1094],[-112,0.106],[-111,0.1029],[-110,0.1021],[-109,0.1021],[-108,0.1004],[-107,0.0959],[-106,0.0951],[-105,0.0924],[-104,0.0954],[-103,0.0987],[-102,0.0991],[-101,0.0981],[-100,0.0953],[-99,0.0971],[-97,0.0943],[-96,0.0957],[-95,0.0963],[-94,0.1005],[-93,0.1004],[-91,0.0955],[-90,0.0958],[-89,0.1001],[-88,0.1037],[-87,0.104],[-86,0.0978],[-85,0.0885],[-84,0.0795],[-83,0.0712],[-81,0.0487],[-80,0.042],[-79,0.0415],[-78,0.0373],[-77,0.029],[-76,0.0258],[-75,0.0249],[-74,0.0215],[-73,0.0215],[-72,0.0229],[-71,0.0229],[-70,0.0225],[-69,0.0234],[-68,0.0239],[-67,0.021],[-66,0.021],[-65,0.024],[-64,0.0239],[-63,0.0203],[-62,0.017],[-61,0.0165],[-60,0.0167],[-59,0.0187],[-58,0.0137],[-57,0.012],[-56,0.0082],[-55,0.0048],[-54,0.0032],[-53,0.0037],[-52,0.0039],[-50,0.0067],[-49,0.0068],[-48,0.0108],[-47,0.0127],[-46,0.013],[-45,0.0142],[-44,0.014],[-43,0.0137],[-42,0.0145],[-41,0.0158],[-40,0.0158],[-39,0.0136],[-38,0.0127],[-37,0.0093],[-36,0.0089],[-35,0.0117],[-34,0.0122],[-33,0.0123],[-32,0.0146],[-31,0.0148],[-30,0.0137],[-29,0.0131],[-28,0.0116],[-27,0.0086],[-25,0.0015],[-24,0.0005],[-23,0.0043],[-22,0.0099],[-21,0.0138],[-20,0.017],[-19,0.0157],[-18,0.0122],[-17,0.0112],[-16,0.0112],[-15,0.0098],[-14,0.0092],[-12,0.0063],[-11,0.0063],[-10,0.0056],[-9,0.0028],[-7,0.0012],[-6,0],[-5,-0.0015],[-4,-0.0027],[-3,-0.0034],[-2,-0.0039],[-1,-0.0027],[0,-0.0024],[1,-0.0032],[2,-0.0027],[3,-0.0019],[4,-0.0015],[5,-0.0012],[6,-0.0002],[7,0],[10,0],[12,-0.0005],[13,-0.0003],[16,0.0014],[17,0.0023],[19,0.002],[20,0.0024],[21,0.0014],[22,0],[23,-0.0003],[24,-0.001],[25,-0.001],[26,-0.0008],[27,-0.0003],[28,-0.0005],[30,0],[31,0.0004],[32,0.0013],[33,0.001],[35,0.001],[36,0.0014],[37,0.001],[38,0.0011],[41,0.0026],[42,0.0027],[43,0.0022],[44,0.0012],[45,0.0007],[46,-0.0003],[47,0.0022],[48,0.0066],[49,0.0088],[50,0.0116],[51,0.0146],[52,0.0165],[53,0.0161],[54,0.0163],[55,0.0172],[57,0.0221],[58,0.0238],[59,0.026],[60,0.0264],[61,0.0263],[62,0.0259],[63,0.0259],[64,0.0264],[66,0.0264],[68,0.0254],[69,0.0244],[71,0.0234],[72,0.022],[73,0.0218],[75,0.022],[76,0.0205],[77,0.0176],[78,0.0175],[80,0.0157],[81,0.0156],[82,0.0152],[83,0.0137],[84,0.0151],[85,0.0151],[86,0.0132],[88,0.0142],[89,0.0129],[90,0.0122],[91,0.0132],[92,0.0137],[94,0.0133],[95,0.0137],[97,0.0139],[98,0.0151],[99,0.0151],[100,0.0141],[101,0.014],[102,0.0142],[103,0.0147],[104,0.0137],[105,0.0137],[106,0.0148],[107,0.0151],[108,0.0152],[109,0.0161],[111,0.02],[114,0.02],[116,0.022],[117,0.022],[118,0.0234],[119,0.0238],[120,0.0235],[121,0.0234],[123,0.0225],[124,0.0229],[125,0.022],[126,0.0205],[128,0.0215],[129,0.0205],[130,0.0186],[131,0.0186],[132,0.0195],[133,0.02],[134,0.022],[135,0.0229],[136,0.0225],[137,0.0239],[138,0.0179],[139,0.0078],[140,0.0013],[141,-0.0004],[142,-0.0012],[143,-0.0002],[144,0.0033],[145,0.0053],[146,0.0039],[147,0.0034],[148,0.0034],[149,0.0049],[150,0.0059],[151,0.0043],[152,0.0034],[153,0.0019],[154,0.0015],[155,0.0033],[156,0.0044],[158,0.0053],[159,0.0039],[161,0.0039],[162,0.0036],[164,0.0054],[166,0.0055],[167,0.0073],[168,0.0111],[169,0.017],[170,0.0191],[171,0.0206],[173,0.0251],[174,0.0275],[175,0.0293],[176,0.0298],[177,0.0297],[178,0.0293],[180,0.0293],[181,0.0281],[182,0.0278],[183,0.0282],[185,0.03],[186,0.0307],[187,0.0303],[189,0.0303],[190,0.0323],[191,0.0332],[192,0.0368],[193,0.045],[194,0.0526],[195,0.0525],[196,0.0531],[197,0.0531],[199,0.0476],[200,0.0469],[201,0.0402],[202,0.0449],[203,0.0461],[204,0.035],[205,0.0335],[207,0.0371],[208,0.0381],[209,0.037],[210,0.0347],[211,0.0337],[212,0.0337],[213,0.0334],[214,0.0337],[215,0.0337],[216,0.0328],[217,0.0327],[218,0.0332],[220,0.0313],[221,0.0312],[223,0.0322],[224,0.0317],[225,0.0317],[226,0.0322],[228,0.0323],[229,0.0332],[231,0.0322],[232,0.0336],[234,0.0395],[235,0.0454],[236,0.0602],[237,0.0678],[238,0.0662],[239,0.0629],[240,0.061],[241,0.0598],[242,0.0596],[243,0.0579],[244,0.0542],[245,0.0501],[246,0.0594],[247,0.0825],[248,0.1107],[249,0.1241],[250,0.1171],[251,0.1129],[253,0.0924],[254,0.0893],[255,0.0889],[256,0.0877],[257,0.0927],[258,0.102],[259,0.1021],[260,0.0989],[261,0.0889],[263,0.0766],[264,0.074],[265,0.0709],[266,0.0729],[267,0.0811],[268,0.0942],[269,0.1054],[270,0.1069],[271,0.1065],[272,0.0983],[273,0.0958],[274,0.0969],[275,0.0995],[276,0.1017],[277,0.1019],[278,0.1011],[279,0.1001],[280,0.1003],[281,0.1019],[282,0.1013],[283,0.1002],[284,0.1001],[285,0.0997],[286,0.1013],[287,0.0997],[288,0.0966],[289,0.0924],[290,0.0923],[291,0.0943],[292,0.0999],[293,0.1106],[294,0.1152],[295,0.1124],[296,0.1081],[297,0.1066],[298,0.0867],[300,0.0495],[301,0.048],[302,0.0361],[303,0.0347],[304,0.0304],[305,0.0285],[306,0.0274],[309,0.026],[310,0.0259],[311,0.0251],[312,0.0249],[313,0.0244],[314,0.0244],[315,0.0249],[316,0.0251],[317,0.0244],[318,0.0234],[319,0.0234],[320,0.0239],[322,0.0239],[323,0.0244],[326,0.023],[328,0.0239],[330,0.0239],[331,0.0234],[332,0.0234],[334,0.0242],[335,0.0235],[336,0.0225],[338,0.0234],[339,0.023],[341,0.0229],[342,0.0225],[343,0.0234],[344,0.0239],[346,0.0239],[347,0.0234],[348,0.0239],[350,0.0239],[352,0.023],[353,0.0222],[357,0.0224],[358,0.0215],[360,0.0215]]
       }
       CTRL_L_eye_eyelidU.X {type=MovieSceneScriptingFloatChannel, default=0, const [0,492]}
       CTRL_L_eye_eyelidU.Y {type=MovieSceneScriptingFloatChannel, default=0, const [0,492]}
       CTRL_L_mouth_cornerSharpnessD {
        type=MovieSceneScriptingFloatChannel
        default=0
        thinned_from=492
        keys [[-131,0.0931],[-130,0.0915],[-129,0.0921],[-128,0.0932],[-127,0.0955],[-126,0.0985],[-125,0.102],[-124,0.1051],[-123,0.107],[-122,0.1074],[-121,0.1088],[-120,0.1066],[-119,0.1041],[-118,0.1021],[-117,0.1017],[-116,0.0989],[-115,0.0966],[-114,0.0971],[-113,0.0958],[-112,0.0935],[-111,0.0928],[-110,0.0913],[-109,0.0882],[-108,0.0834],[-107,0.0796],[-106,0.079],[-105,0.0761],[-104,0.0745],[-103,0.0765],[-102,0.0775],[-101,0.0747],[-100,0.0742],[-99,0.0761],[-98,0.073],[-97,0.0732],[-96,0.0744],[-95,0.0748],[-94,0.0771],[-93,0.0773],[-91,0.0802],[-90,0.0837],[-89,0.0939],[-88,0.1053],[-87,0.1132],[-86,0.114],[-84,0.1129],[-83,0.1101],[-82,0.1013],[-81,0.0907],[-80,0.084],[-79,0.083],[-78,0.0741],[-77,0.0591],[-76,0.053],[-75,0.0479],[-74,0.0415],[-72,0.0404],[-71,0.0391],[-70,0.0386],[-69,0.0376],[-68,0.039],[-66,0.0381],[-65,0.0386],[-64,0.0388],[-63,0.0395],[-62,0.0382],[-61,0.0382],[-60,0.0355],[-59,0.0261],[-58,0.0068],[-57,0.0052],[-56,0.0022],[-54,-0.0011],[-53,-0.0018],[-52,-0.0022],[-51,-0.0004],[-50,0.0009],[-49,0.001],[-48,0.0074],[-47,0.0107],[-46,0.0121],[-45,0.0209],[-44,0.0227],[-42,0.0355],[-41,0.0404],[-40,0.0473],[-39,0.0516],[-38,0.0522],[-37,0.0508],[-36,0.047],[-35,0.0445],[-34,0.0444],[-33,0.0398],[-32,0.0338],[-31,0.0313],[-30,0.0162],[-29,0.0149],[-28,0.0121],[-27,0.01],[-26,0.0083],[-25,0.0057],[-24,0.0048],[-22,-0.0017],[-21,-0.0017],[-20,0.005],[-19,0.0097],[-18,0.0113],[-17,0.0127],[-16,0.0132],[-14,0.0151],[-13,0.0146],[-11,0.0127],[-10,0.0119],[-8,0.0084],[-6,0.0083],[-4,0.0064],[-3,0.0058],[-2,0.0044],[-1,0.0039],[0,0.0044],[1,0.0032],[2,0.0024],[4,0.0025],[5,0.0039],[6,0.0059],[7,0.0063],[8,0.0063],[9,0.0061],[10,0.0054],[11,0.0052],[12,0.0047],[13,0.0044],[14,0.0047],[16,0.0076],[17,0.0082],[19,0.0064],[20,0.0077],[21,0.0083],[22,0.0074],[23,0.0071],[24,0.0064],[25,0.0065],[27,0.0075],[28,0.0075],[29,0.007],[31,0.0073],[32,0.0069],[34,0.0078],[35,0.0078],[37,0.0066],[38,0.007],[39,0.0071],[40,0.0062],[41,0.0059],[43,0.0059],[45,0.0054],[46,0.0034],[47,0.0034],[48,0.0053],[49,0.0064],[50,0.0077],[51,0.0081],[52,0.0092],[53,0.0112],[54,0.0117],[55,0.0134],[56,0.0162],[57,0.0179],[58,0.0184],[59,0.0197],[60,0.02],[61,0.0211],[62,0.0229],[65,0.0229],[66,0.0224],[67,0.0215],[68,0.022],[69,0.0215],[70,0.0195],[71,0.0191],[72,0.0181],[73,0.0189],[74,0.0187],[75,0.0194],[76,0.0194],[77,0.0166],[78,0.0165],[79,0.0151],[80,0.0147],[81,0.0156],[82,0.0152],[83,0.0122],[84,0.0136],[85,0.0141],[86,0.0118],[87,0.0117],[88,0.0122],[90,0.0108],[91,0.0122],[94,0.0111],[95,0.0123],[96,0.0127],[97,0.0127],[98,0.0129],[99,0.0129],[100,0.0127],[102,0.0132],[103,0.0123],[105,0.0122],[106,0.0133],[109,0.0147],[111,0.0176],[112,0.0171],[114,0.0181],[116,0.0181],[117,0.0186],[118,0.0205],[119,0.0205],[120,0.02],[121,0.02],[122,0.019],[124,0.02],[125,0.0182],[126,0.0166],[128,0.0166],[129,0.0152],[130,0.0122],[131,0.0114],[132,0.0122],[133,0.0127],[134,0.0148],[135,0.018],[136,0.022],[137,0.022],[138,0.0132],[139,0.0058],[140,0.0026],[141,0.0015],[142,0.0024],[143,0.0061],[144,0.0101],[146,0.0113],[147,0.0122],[148,0.0117],[150,0.0137],[151,0.0116],[153,0.0088],[154,0.0088],[155,0.0111],[156,0.0117],[157,0.0098],[159,0.0078],[160,0.0079],[161,0.0092],[162,0.0084],[163,0.0093],[164,0.011],[166,0.0093],[167,0.0103],[168,0.0151],[169,0.0218],[170,0.0237],[172,0.0279],[173,0.0284],[175,0.0351],[177,0.0342],[178,0.0326],[179,0.0324],[180,0.0327],[181,0.0322],[182,0.032],[184,0.0309],[186,0.0307],[187,0.0299],[189,0.0298],[192,0.0311],[193,0.0305],[194,0.0308],[195,0.0322],[196,0.036],[197,0.0391],[198,0.0407],[200,0.0405],[201,0.0371],[202,0.0353],[203,0.0346],[204,0.0332],[205,0.0322],[206,0.0325],[207,0.0332],[209,0.0312],[211,0.0283],[212,0.0286],[213,0.0291],[214,0.0288],[216,0.0288],[217,0.0283],[218,0.0288],[219,0.0283],[220,0.0273],[222,0.0264],[223,0.0264],[225,0.0254],[227,0.0264],[229,0.0283],[230,0.0283],[231,0.0298],[232,0.0307],[233,0.0323],[234,0.0327],[235,0.0304],[236,0.033],[237,0.0305],[238,0.0257],[239,0.0274],[240,0.0295],[241,0.031],[242,0.0312],[243,0.0312],[244,0.0298],[245,0.0313],[246,0.0408],[247,0.0649],[248,0.0996],[249,0.1219],[250,0.1155],[251,0.1071],[252,0.0969],[253,0.0904],[255,0.0903],[256,0.0867],[258,0.0957],[259,0.0962],[260,0.0999],[261,0.0975],[262,0.0916],[263,0.0825],[264,0.0809],[265,0.077],[266,0.0757],[267,0.072],[268,0.0671],[269,0.0679],[270,0.0684],[271,0.0704],[272,0.0718],[273,0.0741],[274,0.0794],[275,0.0835],[276,0.0864],[277,0.0877],[278,0.0877],[279,0.0864],[280,0.0869],[281,0.0887],[282,0.0885],[283,0.0867],[284,0.0865],[285,0.0871],[286,0.0892],[287,0.0864],[288,0.0818],[289,0.0762],[290,0.0761],[291,0.0754],[292,0.0724],[293,0.0716],[295,0.0768],[296,0.0763],[297,0.0756],[298,0.0621],[299,0.0508],[300,0.0377],[301,0.037],[302,0.0317],[303,0.031],[304,0.0288],[305,0.0275],[306,0.0266],[307,0.0266],[308,0.0269],[311,0.027],[312,0.0268],[313,0.0259],[316,0.0244],[317,0.0244],[318,0.024],[319,0.0244],[320,0.0258],[321,0.0264],[323,0.0254],[327,0.0273],[328,0.0269],[329,0.0269],[330,0.0273],[332,0.0273],[333,0.0269],[334,0.027],[335,0.0273],[336,0.0269],[337,0.0269],[338,0.0273],[339,0.0273],[340,0.0278],[341,0.0278],[342,0.0274],[343,0.0283],[344,0.0288],[346,0.0288],[347,0.0283],[348,0.0288],[349,0.0289],[351,0.0298],[352,0.0288],[353,0.0283],[354,0.0283],[355,0.028],[356,0.0278],[358,0.0269],[360,0.0269]]
       }
       CTRL_L_eye_eyelidD.X {type=MovieSceneScriptingFloatChannel, default=0, const [0,492]}
       CTRL_L_eye_eyelidD.Y {type=MovieSceneScriptingFloatChannel, default=0, const [0,492]}
       CTRL_L_mouth_cornerSharpnessU {
        type=MovieSceneScriptingFloatChannel
        default=0
        thinned_from=492
        keys [[-131,0.085],[-130,0.0858],[-129,0.088],[-128,0.0885],[-127,0.0899],[-125,0.0939],[-124,0.0955],[-122,0.0952],[-121,0.0961],[-120,0.0931],[-119,0.0895],[-118,0.0879],[-117,0.0874],[-116,0.0839],[-115,0.0817],[-113,0.0826],[-112,0.081],[-111,0.0789],[-109,0.0775],[-108,0.0755],[-107,0.0718],[-106,0.0712],[-105,0.0684],[-104,0.0691],[-103,0.0705],[-102,0.0713],[-101,0.0713],[-100,0.0694],[-99,0.0713],[-97,0.0681],[-96,0.0697],[-95,0.0704],[-94,0.0727],[-93,0.0725],[-92,0.07],[-91,0.0667],[-90,0.067],[-88,0.0761],[-87,0.0784],[-86,0.0744],[-85,0.0688],[-84,0.0625],[-83,0.0546],[-82,0.0425],[-81,0.0307],[-80,0.0264],[-78,0.0264],[-77,0.0259],[-76,0.0297],[-75,0.0375],[-74,0.0375],[-73,0.0361],[-71,0.0391],[-70,0.0409],[-69,0.0415],[-68,0.0406],[-67,0.0391],[-66,0.0395],[-65,0.0418],[-64,0.0432],[-63,0.041],[-62,0.0379],[-61,0.036],[-60,0.0307],[-59,0.0232],[-58,0.0103],[-57,0.009],[-56,0.0063],[-55,0.004],[-54,0.003],[-53,0.004],[-52,0.0044],[-51,0.0043],[-50,0.0053],[-49,0.0054],[-48,0.0093],[-47,0.0102],[-46,0.01],[-45,0.0083],[-44,0.0083],[-43,0.0077],[-42,0.0061],[-41,0.0057],[-40,0.0049],[-38,0.0054],[-37,0.0205],[-36,0.0278],[-35,0.0317],[-34,0.0317],[-33,0.027],[-32,0.0192],[-31,0.0176],[-30,0.0108],[-29,0.01],[-26,0.0054],[-25,0.0032],[-24,0.0024],[-23,0.0029],[-22,0.0064],[-21,0.0106],[-20,0.0106],[-19,0.005],[-18,0.001],[-17,0.001],[-16,0.0015],[-15,0.0015],[-14,0.0009],[-13,-0.0007],[-12,-0.0009],[-11,-0.0002],[-10,-0.0008],[-9,-0.0036],[-8,-0.0048],[-7,-0.0047],[-5,-0.0056],[-3,-0.0071],[-2,-0.0066],[-1,-0.0054],[0,-0.0059],[1,-0.0074],[2,-0.0078],[3,-0.0073],[4,-0.0076],[5,-0.0083],[6,-0.0088],[7,-0.0088],[8,-0.009],[9,-0.0096],[10,-0.0098],[11,-0.0094],[13,-0.0092],[14,-0.0089],[15,-0.0083],[16,-0.0081],[17,-0.0081],[18,-0.0085],[19,-0.0083],[20,-0.0076],[21,-0.0073],[22,-0.008],[23,-0.0084],[24,-0.0096],[25,-0.0101],[26,-0.0099],[27,-0.0089],[28,-0.0082],[30,-0.0084],[31,-0.0077],[32,-0.0076],[33,-0.0078],[35,-0.0074],[36,-0.0064],[37,-0.0066],[38,-0.0065],[39,-0.006],[40,-0.0066],[41,-0.0062],[42,-0.0053],[43,-0.0051],[44,-0.0056],[46,-0.0062],[47,-0.0041],[48,-0.0008],[49,0.0003],[50,0.0023],[51,0.0052],[52,0.0068],[53,0.0074],[54,0.0085],[56,0.0139],[57,0.0158],[59,0.0179],[61,0.0181],[62,0.0195],[65,0.0195],[66,0.019],[71,0.019],[72,0.0181],[73,0.0179],[75,0.0181],[76,0.018],[77,0.0161],[78,0.0161],[79,0.0151],[82,0.0166],[83,0.0156],[84,0.0156],[85,0.0166],[86,0.0161],[87,0.0152],[88,0.0156],[89,0.0156],[90,0.0151],[94,0.015],[95,0.0161],[97,0.0157],[98,0.0159],[100,0.0142],[101,0.0148],[102,0.0151],[103,0.0147],[104,0.0156],[105,0.0156],[106,0.0171],[108,0.0171],[109,0.0176],[111,0.0205],[112,0.02],[113,0.02],[115,0.022],[116,0.022],[117,0.021],[118,0.0205],[120,0.0215],[121,0.0225],[122,0.0229],[123,0.0215],[124,0.0215],[125,0.022],[126,0.0215],[127,0.0215],[128,0.021],[130,0.0181],[131,0.0185],[133,0.0215],[134,0.0219],[135,0.021],[136,0.0211],[137,0.0229],[138,0.0205],[139,0.0131],[140,0.0076],[141,0.0071],[142,0.0073],[143,0.0078],[144,0.0107],[145,0.0122],[146,0.0108],[147,0.0103],[148,0.0103],[150,0.0141],[151,0.0126],[152,0.0117],[153,0.0102],[154,0.0093],[155,0.0107],[158,0.0137],[160,0.0117],[161,0.0117],[162,0.0113],[164,0.0122],[165,0.0122],[167,0.0142],[168,0.0163],[169,0.0198],[170,0.0199],[171,0.0195],[172,0.0207],[173,0.0214],[174,0.0219],[175,0.0229],[176,0.0234],[177,0.0232],[179,0.0212],[180,0.0205],[181,0.019],[182,0.0185],[183,0.0178],[184,0.0191],[185,0.0199],[186,0.0192],[187,0.0174],[189,0.0166],[190,0.0176],[191,0.0176],[192,0.0187],[193,0.0247],[194,0.0349],[195,0.0354],[196,0.0344],[198,0.03],[199,0.0293],[200,0.0293],[201,0.0268],[202,0.0317],[203,0.0285],[204,0.016],[205,0.0139],[206,0.0151],[207,0.0166],[208,0.0171],[209,0.0165],[210,0.0151],[213,0.0152],[214,0.0156],[215,0.0152],[216,0.0143],[217,0.0146],[218,0.0156],[220,0.0147],[221,0.0151],[222,0.0151],[223,0.0142],[224,0.0137],[226,0.0146],[228,0.0147],[229,0.0156],[230,0.0147],[231,0.0142],[232,0.0151],[233,0.0158],[234,0.0171],[235,0.0199],[236,0.0329],[237,0.0397],[238,0.0388],[239,0.0344],[240,0.0303],[241,0.0277],[242,0.0273],[243,0.0263],[244,0.0249],[245,0.0222],[246,0.028],[247,0.0453],[248,0.0676],[249,0.0778],[250,0.0727],[251,0.0713],[252,0.0662],[253,0.0606],[254,0.0581],[255,0.0576],[256,0.0563],[257,0.0588],[258,0.063],[259,0.063],[260,0.0612],[261,0.0549],[263,0.0468],[265,0.0443],[266,0.0463],[267,0.0533],[268,0.0634],[269,0.0721],[270,0.0732],[271,0.0735],[272,0.0688],[273,0.0669],[274,0.0674],[276,0.0716],[277,0.0727],[278,0.0735],[279,0.0728],[280,0.0729],[281,0.0746],[282,0.0744],[283,0.0725],[284,0.0723],[285,0.0723],[286,0.0736],[287,0.0713],[289,0.0642],[290,0.064],[291,0.065],[292,0.0687],[293,0.0758],[294,0.0791],[295,0.0779],[296,0.0739],[297,0.0725],[298,0.0538],[299,0.0393],[300,0.0265],[301,0.0258],[302,0.0205],[303,0.0198],[304,0.0179],[305,0.0176],[307,0.0176],[309,0.0181],[310,0.018],[311,0.0172],[312,0.0171],[313,0.0166],[314,0.0166],[316,0.0174],[317,0.0171],[318,0.0171],[319,0.0181],[321,0.0171],[322,0.018],[323,0.0181],[324,0.0176],[325,0.0181],[326,0.0176],[327,0.0176],[328,0.0181],[329,0.018],[331,0.0171],[332,0.0171],[333,0.0176],[334,0.0184],[335,0.0176],[336,0.0166],[337,0.0166],[338,0.0171],[341,0.0171],[342,0.0166],[344,0.0176],[347,0.0176],[348,0.0181],[349,0.018],[352,0.0166],[353,0.0159],[356,0.0161],[357,0.0156],[358,0.0156],[359,0.0161],[360,0.0161]]
       }
       CTRL_R_mouth_lipsTowardsTeethD {type=MovieSceneScriptingFloatChannel, default=0, const [0,492]}
       CTRL_L_mouth_lipsTowardsTeethD {type=MovieSceneScriptingFloatChannel, default=0, const [0,492]}
       CTRL_R_eye_eyelidU.X {type=MovieSceneScriptingFloatChannel, default=0, const [0,492]}
       CTRL_R_eye_eyelidU.Y {type=MovieSceneScriptingFloatChannel, default=0, const [0,492]}
       CTRL_L_mouth_lipsTowardsTeethU {type=MovieSceneScriptingFloatChannel, default=0, const [0,492]}
       CTRL_R_eye_eyelidD.X {type=MovieSceneScriptingFloatChannel, default=0, const [0,492]}
       CTRL_R_eye_eyelidD.Y {type=MovieSceneScriptingFloatChannel, default=0, const [0,492]}
       CTRL_R_mouth_lipsTowardsTeethU {type=MovieSceneScriptingFloatChannel, default=0, const [0,492]}
       CTRL_faceGUIfollowHead {type=MovieSceneScriptingFloatChannel, default=1, const [1,492]}
       CTRL_eyesAimFollowHead {type=MovieSceneScriptingFloatChannel, default=0, const [0,492]}
       CTRL_C_nose.X {type=MovieSceneScriptingFloatChannel, default=0, const [0,492]}
       CTRL_C_nose.Y {type=MovieSceneScriptingFloatChannel, default=0, const [0,492]}
       CTRL_C_teeth_rotateU {type=MovieSceneScriptingFloatChannel, default=0, const [0,492]}
       CTRL_C_teeth_rotateD {type=MovieSceneScriptingFloatChannel, default=0, const [0,492]}
       CTRL_L_brow_inner {type=MovieSceneScriptingFloatChannel, default=0, const [0,492]}
       CTRL_R_brow_inner {type=MovieSceneScriptingFloatChannel, default=0, const [0,492]}
       CTRL_L_brow_outer {type=MovieSceneScriptingFloatChannel, default=0, const [0,492]}
       CTRL_R_brow_outer {type=MovieSceneScriptingFloatChannel, default=0, const [0,492]}
       CTRL_C_mouth.X {
        type=MovieSceneScriptingFloatChannel
        default=0
        thinned_from=492
        keys [[-131,0.0414],[-129,0.0411],[-128,0.0406],[-127,0.0396],[-126,0.0394],[-125,0.0387],[-124,0.0391],[-123,0.0384],[-122,0.0381],[-121,0.0407],[-120,0.0424],[-119,0.0432],[-117,0.0431],[-116,0.0433],[-114,0.0453],[-113,0.0454],[-112,0.0458],[-111,0.0468],[-108,0.0477],[-107,0.0487],[-106,0.0488],[-105,0.0479],[-104,0.0494],[-103,0.0489],[-102,0.048],[-101,0.0479],[-100,0.0483],[-99,0.0469],[-98,0.0462],[-97,0.0463],[-96,0.0458],[-95,0.0456],[-93,0.0459],[-92,0.0477],[-91,0.0489],[-90,0.048],[-89,0.0465],[-88,0.0464],[-87,0.0456],[-85,0.0417],[-82,0.0366],[-81,0.0382],[-80,0.0403],[-79,0.04],[-78,0.0373],[-77,0.0371],[-76,0.0352],[-75,0.0362],[-74,0.0355],[-73,0.0329],[-72,0.0343],[-71,0.0369],[-70,0.0385],[-69,0.0357],[-68,0.0355],[-67,0.0322],[-66,0.0312],[-65,0.0294],[-63,0.0331],[-62,0.0306],[-61,0.0286],[-60,0.026],[-59,0.0194],[-58,0.0102],[-57,0.0094],[-56,0.0072],[-55,0.0045],[-54,0.0043],[-53,0.0051],[-52,0.0054],[-51,0.0065],[-50,0.0059],[-49,0.0058],[-48,0.0072],[-47,0.008],[-46,0.008],[-45,0.0083],[-44,0.0084],[-43,0.0089],[-42,0.0104],[-41,0.0124],[-40,0.0158],[-39,0.0208],[-38,0.0225],[-37,0.0318],[-36,0.0369],[-35,0.0329],[-34,0.0273],[-33,0.0247],[-32,0.0199],[-31,0.0182],[-30,0.0099],[-29,0.0092],[-28,0.0075],[-27,0.0061],[-26,0.007],[-25,0.0078],[-24,0.0078],[-23,0.0103],[-22,0.0106],[-21,0.0114],[-20,0.0083],[-19,0.0096],[-18,0.0097],[-16,0.0125],[-15,0.0121],[-14,0.0124],[-13,0.0121],[-12,0.0114],[-11,0.012],[-10,0.0115],[-9,0.0096],[-8,0.0093],[-7,0.0101],[-6,0.0104],[-4,0.0119],[-3,0.0117],[-2,0.0098],[-1,0.0107],[0,0.0098],[1,0.0106],[2,0.0113],[3,0.0115],[4,0.0121],[5,0.012],[6,0.013],[7,0.0148],[8,0.0147],[9,0.0143],[10,0.0145],[12,0.0142],[13,0.0138],[14,0.014],[15,0.0147],[16,0.0143],[17,0.0131],[18,0.0128],[19,0.0132],[20,0.013],[22,0.0115],[24,0.0118],[25,0.0108],[26,0.0101],[27,0.01],[28,0.0111],[29,0.0115],[30,0.0123],[32,0.0113],[34,0.0116],[35,0.0115],[36,0.0113],[37,0.0119],[38,0.0113],[39,0.0104],[40,0.0112],[41,0.0116],[42,0.0109],[43,0.0097],[44,0.0092],[45,0.0093],[46,0.0111],[47,0.0101],[48,0.0111],[49,0.0114],[50,0.0129],[51,0.0146],[52,0.0154],[54,0.0152],[55,0.0167],[57,0.0202],[58,0.021],[59,0.0221],[60,0.0223],[61,0.024],[62,0.0244],[63,0.0246],[64,0.0268],[65,0.0268],[67,0.0252],[68,0.0279],[69,0.0266],[70,0.0255],[71,0.0255],[72,0.0233],[73,0.0228],[74,0.0218],[75,0.0204],[76,0.0188],[77,0.0181],[78,0.0181],[80,0.0185],[81,0.019],[82,0.0193],[83,0.0181],[84,0.0192],[86,0.0195],[87,0.0174],[88,0.018],[89,0.017],[90,0.0172],[91,0.0169],[92,0.0176],[93,0.0173],[94,0.0167],[95,0.0165],[96,0.0152],[97,0.0149],[98,0.015],[99,0.0147],[100,0.0146],[101,0.0152],[103,0.0152],[105,0.0154],[106,0.0173],[107,0.0174],[108,0.018],[109,0.0178],[110,0.0172],[112,0.0194],[113,0.0244],[114,0.0259],[115,0.0251],[116,0.0253],[117,0.026],[118,0.0302],[119,0.0322],[120,0.0333],[123,0.0331],[124,0.0306],[125,0.0305],[126,0.029],[127,0.0313],[128,0.0323],[130,0.0322],[131,0.0314],[132,0.0322],[133,0.0327],[134,0.0353],[135,0.0426],[136,0.0479],[137,0.0538],[138,0.0462],[139,0.0303],[140,0.0225],[141,0.0205],[142,0.0197],[143,0.0228],[144,0.025],[146,0.0253],[147,0.0258],[148,0.0267],[149,0.0266],[150,0.0281],[152,0.0263],[153,0.0243],[154,0.0248],[155,0.0261],[156,0.0258],[157,0.0244],[159,0.0244],[160,0.024],[161,0.0249],[162,0.0228],[163,0.0235],[164,0.0226],[165,0.0236],[166,0.0224],[167,0.0215],[168,0.0229],[169,0.0279],[170,0.0283],[171,0.0283],[173,0.0312],[174,0.0358],[175,0.0382],[176,0.0376],[177,0.0368],[178,0.0335],[179,0.0313],[180,0.0309],[181,0.0313],[182,0.032],[183,0.0317],[184,0.0328],[187,0.039],[188,0.0394],[189,0.0429],[190,0.0471],[191,0.0505],[192,0.0566],[193,0.0634],[194,0.0663],[195,0.0622],[196,0.0551],[197,0.052],[198,0.0522],[199,0.0499],[200,0.0487],[201,0.0454],[202,0.0543],[203,0.046],[204,0.0428],[205,0.0447],[206,0.044],[207,0.0439],[209,0.0409],[210,0.0369],[211,0.0359],[213,0.0376],[214,0.0392],[215,0.0383],[216,0.0358],[217,0.0364],[218,0.0372],[219,0.0351],[220,0.0346],[221,0.0354],[222,0.0342],[223,0.0327],[224,0.0325],[225,0.0333],[227,0.0332],[228,0.0344],[229,0.0372],[230,0.0381],[231,0.0393],[232,0.0419],[233,0.0457],[234,0.0449],[235,0.0407],[236,0.0435],[237,0.0429],[238,0.0386],[239,0.0324],[240,0.0289],[241,0.0283],[242,0.0283],[243,0.0328],[245,0.0391],[246,0.0461],[247,0.0559],[248,0.0624],[249,0.0663],[250,0.0646],[251,0.0627],[252,0.0619],[253,0.0594],[254,0.0581],[255,0.0598],[256,0.0574],[257,0.0568],[258,0.0597],[259,0.0588],[260,0.0493],[261,0.0421],[262,0.0359],[263,0.0318],[264,0.031],[265,0.0296],[267,0.0294],[268,0.0308],[269,0.0313],[270,0.0313],[272,0.0293],[273,0.0315],[274,0.0346],[275,0.038],[276,0.0399],[277,0.0414],[278,0.0424],[279,0.0436],[281,0.0444],[282,0.0442],[283,0.0432],[284,0.0431],[286,0.045],[288,0.0436],[292,0.0435],[293,0.045],[294,0.0458],[295,0.0455],[297,0.0456],[298,0.0454],[299,0.0437],[300,0.0443],[301,0.0438],[302,0.0389],[303,0.0379],[304,0.0347],[305,0.032],[306,0.0288],[307,0.0269],[308,0.0269],[309,0.0275],[310,0.0276],[311,0.0241],[312,0.0235],[313,0.0245],[314,0.0244],[315,0.024],[316,0.0245],[317,0.024],[319,0.0243],[320,0.0234],[321,0.0243],[322,0.0243],[323,0.0249],[325,0.0233],[326,0.0228],[327,0.0222],[328,0.0218],[330,0.0204],[331,0.0213],[332,0.0209],[333,0.0202],[334,0.0213],[335,0.0214],[336,0.0203],[337,0.0204],[338,0.0199],[339,0.0202],[341,0.0193],[342,0.0194],[343,0.0203],[344,0.0199],[345,0.0197],[347,0.0216],[348,0.022],[349,0.0213],[350,0.0214],[351,0.0209],[352,0.0212],[353,0.0212],[354,0.0216],[355,0.0212],[356,0.021],[357,0.0201],[358,0.0189],[359,0.0192],[360,0.0192]]
       }
       CTRL_C_mouth.Y {type=MovieSceneScriptingFloatChannel, default=0, const [0,492]}
       CTRL_L_mouth_upperLipRaise {
        type=MovieSceneScriptingFloatChannel
        default=0
        thinned_from=492
        keys [[-131,0.0026],[-130,0.0012],[-129,0.0012],[-126,0],[-125,0.0003],[-124,0.0004],[-121,0],[-120,0.0007],[-119,0.0003],[-118,0.0007],[-117,0.0008],[-116,0.0012],[-115,0.0013],[-113,0.0009],[-111,0.0009],[-110,0.0005],[-109,0.0006],[-107,0.0013],[-106,0.0013],[-105,0.0011],[-104,0.0013],[-103,0.0011],[-101,0.001],[-100,0.0005],[-95,0.0011],[-94,0.0008],[-92,0.0007],[-91,0.0004],[-90,0.001],[-88,0.0016],[-87,0.0022],[-86,0.0025],[-84,0.0028],[-83,0.0031],[-79,0.0032],[-77,0.0027],[-76,0.0028],[-74,0.0026],[-73,0.003],[-72,0.0029],[-71,0.0026],[-68,0.0028],[-67,0.0031],[-66,0.003],[-65,0.0026],[-59,0.0028],[-54,0.0023],[-49,0.0026],[-46,0.0027],[-45,0.003],[-38,0.0029],[-31,0.0029],[-30,0.0025],[-24,0.0022],[-23,0.0019],[-22,0.0026],[-20,0.003],[-19,0.0028],[-17,0.0029],[-16,0.0033],[-14,0.0032],[-13,0.0033],[-11,0.0032],[-10,0.0033],[-8,0.0032],[-6,0.0034],[-5,0.0032],[-4,0.0033],[-3,0.0032],[-1,0.0034],[0,0.0032],[2,0.0034],[3,0.0033],[4,0.0035],[6,0.0031],[16,0.0032],[20,0.003],[21,0.0031],[22,0.0028],[25,0.0031],[28,0.0029],[30,0.0032],[33,0.0028],[41,0.0029],[43,0.0031],[45,0.003],[46,0.0033],[51,0.0025],[52,0.0026],[53,0.0024],[55,0.0026],[56,0.0024],[58,0.0007],[59,0.0011],[62,0.0017],[65,0.0016],[66,0.0022],[67,0.0017],[68,0.0016],[69,0.002],[72,0.0017],[73,0.0014],[74,0.0014],[75,0.0018],[76,0.0015],[78,0.0016],[79,0.001],[81,0.001],[82,0.0007],[84,0.0006],[85,0.0007],[87,0.0015],[88,0.0012],[89,0.0013],[91,0.0009],[97,0.0009],[98,0.0005],[100,0.0015],[103,0.0012],[104,0.0009],[106,0.0007],[107,0.001],[108,0.0003],[110,0.0008],[111,0.0009],[113,0.0007],[114,0.0003],[115,0.0009],[116,0.0011],[117,0.0007],[118,0.0007],[119,0.0004],[120,0.0008],[122,0.0009],[123,0.0005],[124,0.0006],[125,0.0003],[127,0.0004],[128,0.0001],[129,0.0002],[130,0.0006],[132,0],[134,0.0011],[136,0.001],[137,0.0036],[138,0.0177],[139,0.0147],[140,0.0205],[141,0.019],[143,0.0145],[144,0.0081],[145,0.0057],[146,0.0074],[147,0.0058],[148,0.0074],[149,0.0041],[150,0.0038],[151,0.0066],[152,0.0071],[153,0.0106],[154,0.0102],[155,0.008],[156,0.0086],[157,0.0081],[159,0.0117],[160,0.0106],[161,0.011],[162,0.0109],[163,0.0094],[164,0.0081],[165,0.0077],[166,0.0075],[167,0.0053],[168,0.0026],[170,0.0026],[171,0.0031],[172,0.0027],[173,0.0018],[174,0.0012],[175,0.0014],[176,0.0018],[178,0.0017],[179,0.0008],[180,0.0009],[181,0.0013],[182,0.0012],[183,0.0008],[184,0.0007],[185,0.0011],[188,0.0014],[190,0.0009],[191,0.0009],[192,0.0005],[194,0.0027],[195,0.0027],[196,0.0024],[197,0.0018],[198,0.0026],[199,0.0032],[200,0.0032],[201,0.0029],[202,0.0036],[203,0.0032],[207,0.0026],[208,0.0026],[210,0.0025],[211,0.0028],[213,0.0024],[216,0.0026],[217,0.0022],[219,0.0024],[220,0.0029],[222,0.0027],[223,0.0029],[224,0.0028],[227,0.0031],[228,0.0026],[231,0.0025],[233,0.0028],[234,0.0026],[235,0.0013],[236,0.0024],[237,0.0027],[238,0.0026],[239,0.0015],[240,0.0015],[242,0.0017],[243,0.001],[244,0.0008],[245,0.0012],[246,0.0006],[247,0.0006],[248,0.0002],[249,0.0006],[250,0.0002],[251,0.0007],[253,0.0003],[254,0.0011],[256,0.0006],[257,0],[259,0.0002],[260,0.0015],[261,0.0022],[262,0.002],[263,0.0029],[265,0.0033],[268,0.0031],[269,0.0026],[270,0.0025],[271,0.0026],[274,0.001],[276,0.0005],[279,0.0005],[281,0.0011],[282,0.0008],[284,0.001],[285,0.0009],[286,0.0002],[288,0.0006],[289,0.0007],[291,0.0002],[293,0.0005],[294,0.0005],[296,0.0011],[298,0.0012],[300,0.0024],[301,0.0024],[302,0.002],[307,0.0027],[308,0.0024],[313,0.0022],[315,0.002],[316,0.002],[318,0.0018],[319,0.0019],[320,0.0023],[321,0.0022],[322,0.0018],[323,0.0016],[324,0.0022],[326,0.0024],[327,0.002],[331,0.0023],[334,0.002],[336,0.0022],[337,0.0019],[338,0.0021],[339,0.0019],[340,0.0021],[342,0.0018],[343,0.002],[344,0.0018],[345,0.0019],[348,0.0017],[349,0.0018],[350,0.0016],[351,0.002],[352,0.002],[353,0.0023],[360,0.0019]]
       }
       CTRL_L_mouth_cornerPull {
        type=MovieSceneScriptingFloatChannel
        default=0.5472
        thinned_from=492
        keys [[-131,0.0018],[-129,0.0012],[-128,0.0005],[-126,0],[-122,0.0002],[-121,0],[-120,0.0004],[-116,0.0004],[-114,0.0007],[-111,0.0007],[-109,0.0002],[-108,0.0004],[-107,0.0008],[-105,0.0011],[-99,0.0003],[-95,0.0009],[-92,0.0006],[-91,0.0003],[-85,0.0023],[-82,0.0028],[-79,0.0028],[-77,0.0026],[-75,0.0026],[-73,0.0023],[-60,0.0022],[-59,0.0028],[-58,0.0072],[-57,0.0097],[-56,0.0167],[-55,0.0242],[-54,0.029],[-53,0.0316],[-51,0.0339],[-50,0.0326],[-49,0.0325],[-48,0.0319],[-47,0.0302],[-46,0.0297],[-44,0.0293],[-43,0.0277],[-42,0.0215],[-41,0.0155],[-40,0.0104],[-39,0.0069],[-38,0.0061],[-37,0.0029],[-34,0.0025],[-33,0.0026],[-32,0.0042],[-31,0.0067],[-30,0.0239],[-29,0.0269],[-28,0.0338],[-27,0.0385],[-26,0.044],[-25,0.0498],[-24,0.0514],[-23,0.0478],[-22,0.04],[-21,0.0285],[-20,0.0276],[-19,0.0284],[-18,0.0328],[-17,0.032],[-16,0.0337],[-15,0.0346],[-14,0.0372],[-13,0.0384],[-12,0.0401],[-11,0.0413],[-10,0.0407],[-9,0.0414],[-8,0.0425],[-7,0.0449],[-6,0.0452],[-5,0.0482],[-4,0.0503],[-3,0.0533],[-2,0.0553],[0,0.0561],[1,0.0546],[3,0.0563],[4,0.0581],[5,0.0568],[6,0.0553],[7,0.0564],[9,0.0568],[10,0.0562],[11,0.0562],[12,0.0557],[13,0.0549],[14,0.0532],[15,0.0529],[16,0.0529],[17,0.0518],[19,0.0482],[21,0.0502],[22,0.0518],[23,0.0521],[24,0.0528],[25,0.0531],[27,0.0524],[28,0.0526],[29,0.0536],[30,0.0543],[31,0.0546],[33,0.0526],[35,0.0527],[36,0.0533],[37,0.0537],[38,0.0534],[39,0.0527],[40,0.0508],[41,0.0494],[42,0.0495],[43,0.0505],[44,0.0512],[46,0.0514],[47,0.0498],[48,0.0434],[50,0.0343],[51,0.0294],[52,0.0262],[53,0.0272],[54,0.0311],[55,0.0395],[56,0.0442],[57,0.0465],[58,0.0462],[59,0.0489],[60,0.0497],[61,0.0509],[62,0.0571],[63,0.0576],[64,0.0649],[65,0.0649],[66,0.0685],[67,0.0679],[68,0.0652],[69,0.0657],[70,0.0675],[71,0.0706],[72,0.0724],[75,0.083],[76,0.0876],[77,0.0942],[78,0.0945],[79,0.0994],[80,0.1076],[81,0.1114],[82,0.1193],[83,0.1211],[84,0.1195],[85,0.1164],[86,0.119],[87,0.1205],[88,0.1208],[89,0.116],[90,0.1176],[91,0.1157],[92,0.1108],[93,0.111],[95,0.113],[96,0.1131],[97,0.1119],[98,0.1075],[99,0.1068],[101,0.112],[102,0.1128],[103,0.11],[104,0.1118],[105,0.1119],[106,0.1125],[107,0.1114],[108,0.1125],[109,0.1112],[110,0.107],[111,0.1013],[112,0.1023],[113,0.1024],[114,0.1051],[115,0.1046],[116,0.1103],[117,0.1126],[118,0.1109],[119,0.1049],[120,0.1002],[121,0.1093],[122,0.1212],[123,0.1276],[124,0.1283],[125,0.1273],[126,0.1315],[128,0.1344],[129,0.139],[130,0.1456],[131,0.1484],[132,0.1517],[133,0.1595],[134,0.1758],[135,0.1927],[136,0.2053],[137,0.2188],[138,0.2521],[139,0.2837],[140,0.3135],[141,0.3198],[142,0.3349],[143,0.338],[144,0.3348],[145,0.3207],[146,0.3171],[147,0.316],[148,0.3227],[149,0.3169],[150,0.3195],[151,0.3212],[152,0.3244],[153,0.3241],[154,0.3226],[155,0.3203],[156,0.3198],[157,0.3186],[158,0.3195],[159,0.3143],[160,0.3082],[161,0.307],[162,0.3092],[163,0.3129],[164,0.3091],[165,0.3084],[167,0.2948],[168,0.2758],[169,0.255],[170,0.2432],[171,0.2276],[172,0.2104],[173,0.1904],[174,0.1689],[175,0.1468],[176,0.1299],[177,0.1276],[178,0.1203],[179,0.1152],[180,0.1142],[181,0.1086],[182,0.1049],[183,0.1013],[184,0.1003],[185,0.1],[187,0.0953],[188,0.0948],[189,0.0891],[191,0.077],[192,0.0707],[193,0.0683],[194,0.0657],[195,0.0652],[196,0.0577],[197,0.0476],[198,0.0371],[199,0.0331],[200,0.033],[201,0.0331],[202,0.0351],[203,0.0348],[204,0.035],[205,0.029],[206,0.027],[207,0.0256],[208,0.0259],[209,0.0255],[210,0.0277],[211,0.0309],[212,0.0327],[213,0.0349],[214,0.0343],[216,0.0358],[217,0.0377],[219,0.039],[220,0.0382],[221,0.0369],[224,0.0316],[225,0.0295],[226,0.0287],[227,0.0274],[228,0.0255],[229,0.0238],[230,0.0229],[232,0.0227],[233,0.0212],[234,0.0199],[235,0.0176],[236,0.0168],[237,0.0151],[238,0.0122],[239,0.0088],[240,0.0062],[241,0.004],[242,0.0037],[243,0.0014],[244,0.0008],[246,0.0005],[247,0.0005],[248,0.0002],[249,0.0004],[250,0.0002],[251,0.0004],[252,0.0004],[254,0.0038],[255,0.0058],[256,0.0054],[258,0.0019],[259,0.0018],[260,0.001],[262,0.0017],[263,0.0029],[264,0.0037],[265,0.0054],[266,0.0063],[268,0.0048],[269,0.0038],[271,0.0033],[272,0.0021],[274,0.0007],[275,0.0003],[276,0],[280,0],[285,0.0002],[292,0],[294,0.0002],[296,0.0008],[297,0.0008],[298,0.0013],[299,0.0019],[300,0.0044],[301,0.0051],[302,0.0124],[303,0.0142],[304,0.0199],[305,0.0231],[306,0.0253],[308,0.0269],[310,0.0267],[311,0.0263],[312,0.0262],[313,0.0267],[314,0.0287],[315,0.029],[317,0.0286],[318,0.0297],[319,0.0303],[320,0.0304],[321,0.0301],[322,0.0292],[323,0.0277],[324,0.0269],[325,0.0277],[326,0.0276],[328,0.0259],[329,0.0257],[330,0.0253],[331,0.0246],[335,0.0249],[337,0.0261],[338,0.0265],[339,0.0261],[340,0.0261],[341,0.0267],[342,0.0279],[343,0.0285],[344,0.0281],[346,0.0285],[348,0.0282],[349,0.0296],[350,0.0312],[351,0.0335],[355,0.0385],[358,0.0394],[360,0.0394]]
       }
       CTRL_L_mouth_dimple {
        type=MovieSceneScriptingFloatChannel
        default=0
        thinned_from=492
        keys [[-131,0.0337],[-130,0.0266],[-129,0.0204],[-128,0.018],[-127,0.0194],[-126,0.0211],[-124,0.0263],[-123,0.0283],[-121,0.0292],[-120,0.0295],[-119,0.0295],[-118,0.0304],[-117,0.0305],[-116,0.0315],[-115,0.0313],[-114,0.0294],[-112,0.0228],[-111,0.0199],[-110,0.0183],[-108,0.0143],[-107,0.0127],[-106,0.0125],[-105,0.0115],[-104,0.0107],[-103,0.011],[-102,0.0106],[-101,0.009],[-100,0.0089],[-99,0.007],[-98,0.0061],[-96,0.0035],[-95,0.0032],[-94,0.0024],[-93,0.0023],[-91,0.003],[-90,0.0054],[-89,0.0118],[-88,0.0202],[-86,0.0409],[-85,0.0507],[-84,0.063],[-83,0.0757],[-82,0.0914],[-81,0.1038],[-80,0.12],[-79,0.1218],[-78,0.1377],[-76,0.1836],[-75,0.2077],[-74,0.2344],[-73,0.2576],[-71,0.2843],[-70,0.2865],[-69,0.2842],[-68,0.2848],[-67,0.2795],[-65,0.2803],[-64,0.2843],[-63,0.2859],[-62,0.2888],[-61,0.2892],[-60,0.2824],[-59,0.2668],[-58,0.2311],[-57,0.2218],[-56,0.2002],[-55,0.1853],[-54,0.1747],[-53,0.1687],[-52,0.1662],[-51,0.1509],[-50,0.1409],[-49,0.14],[-48,0.1231],[-47,0.1237],[-46,0.1249],[-45,0.1347],[-44,0.1381],[-43,0.1502],[-42,0.1604],[-41,0.1685],[-40,0.1823],[-39,0.1985],[-38,0.2034],[-37,0.2479],[-36,0.2751],[-35,0.3043],[-34,0.3175],[-33,0.3158],[-32,0.2981],[-31,0.2903],[-30,0.24],[-29,0.2297],[-28,0.2055],[-27,0.1902],[-26,0.1842],[-25,0.1836],[-24,0.1835],[-23,0.1718],[-22,0.1565],[-21,0.138],[-20,0.1222],[-19,0.1131],[-18,0.1068],[-17,0.1054],[-16,0.1006],[-15,0.0999],[-14,0.0987],[-13,0.0983],[-12,0.0974],[-11,0.0967],[-10,0.0987],[-9,0.0971],[-8,0.0968],[-7,0.098],[-6,0.0961],[-5,0.0982],[-4,0.0986],[-3,0.0969],[-2,0.0979],[-1,0.0949],[0,0.0938],[1,0.092],[2,0.089],[3,0.0884],[4,0.087],[5,0.0873],[6,0.0869],[7,0.0874],[8,0.0872],[10,0.0856],[12,0.0856],[14,0.0861],[16,0.0854],[17,0.0854],[19,0.0837],[20,0.083],[21,0.0839],[22,0.0852],[23,0.0855],[24,0.0869],[25,0.0892],[26,0.09],[27,0.0918],[28,0.0923],[29,0.0921],[31,0.0912],[33,0.0915],[34,0.0911],[35,0.0908],[36,0.0892],[37,0.0889],[38,0.0892],[39,0.09],[40,0.09],[41,0.0885],[42,0.0868],[43,0.0859],[44,0.0871],[45,0.0875],[46,0.0872],[47,0.0882],[48,0.0861],[49,0.0858],[50,0.0851],[51,0.0824],[52,0.0815],[53,0.0943],[54,0.1013],[55,0.1189],[56,0.1323],[57,0.1429],[58,0.1457],[59,0.145],[60,0.1448],[61,0.1468],[62,0.1506],[63,0.151],[64,0.1575],[65,0.1575],[66,0.159],[67,0.158],[68,0.1553],[69,0.152],[70,0.147],[71,0.1451],[72,0.1451],[73,0.1497],[74,0.1577],[75,0.166],[76,0.1776],[77,0.1848],[78,0.185],[79,0.1884],[80,0.1907],[81,0.1917],[82,0.1923],[83,0.1915],[84,0.1915],[85,0.1886],[86,0.1888],[87,0.1879],[88,0.1856],[89,0.1839],[90,0.1817],[91,0.1802],[92,0.1774],[93,0.1764],[94,0.1733],[95,0.1713],[96,0.1707],[98,0.1688],[100,0.1661],[101,0.1651],[102,0.1646],[103,0.1633],[104,0.1648],[105,0.1649],[106,0.1667],[107,0.1674],[109,0.1703],[110,0.1738],[111,0.1813],[112,0.1857],[114,0.1847],[115,0.1856],[116,0.19],[118,0.2105],[119,0.2175],[122,0.2296],[123,0.2309],[124,0.2339],[125,0.2326],[126,0.2295],[127,0.2281],[128,0.225],[129,0.2242],[130,0.2187],[131,0.2169],[132,0.2158],[133,0.2159],[134,0.2232],[135,0.2331],[136,0.2471],[137,0.2563],[138,0.2619],[140,0.2794],[141,0.2886],[142,0.3159],[143,0.3208],[144,0.3215],[145,0.3137],[146,0.3048],[147,0.3033],[150,0.3111],[151,0.3106],[152,0.3085],[153,0.3083],[154,0.3091],[155,0.3065],[156,0.3044],[157,0.3033],[158,0.2981],[159,0.2965],[160,0.2911],[161,0.2897],[162,0.2913],[163,0.2922],[164,0.2918],[165,0.2905],[166,0.2904],[167,0.2888],[168,0.285],[170,0.2665],[171,0.2597],[172,0.2505],[173,0.2391],[174,0.2299],[175,0.2183],[176,0.2121],[177,0.2107],[178,0.2061],[179,0.2024],[180,0.1999],[181,0.1922],[183,0.1778],[184,0.1738],[185,0.1703],[186,0.166],[187,0.1588],[188,0.1575],[189,0.1503],[190,0.1402],[191,0.1383],[192,0.137],[193,0.1385],[194,0.1465],[195,0.1489],[196,0.1496],[197,0.1404],[198,0.1294],[199,0.1211],[200,0.1188],[201,0.1261],[202,0.1196],[203,0.1235],[204,0.1254],[205,0.1196],[206,0.1276],[207,0.125],[208,0.1204],[209,0.1184],[210,0.1192],[211,0.1211],[212,0.1208],[213,0.1216],[214,0.1194],[215,0.1183],[216,0.1189],[217,0.117],[218,0.118],[219,0.1183],[220,0.1163],[221,0.1196],[223,0.1159],[224,0.1133],[225,0.1088],[226,0.108],[227,0.107],[228,0.1075],[229,0.107],[230,0.1086],[231,0.1094],[232,0.1078],[233,0.1073],[234,0.1035],[235,0.1],[236,0.0945],[237,0.0848],[238,0.0769],[239,0.0673],[240,0.0635],[241,0.0589],[242,0.0579],[243,0.0527],[244,0.0479],[245,0.0426],[246,0.0334],[247,0.0263],[248,0.0207],[249,0.0173],[250,0.0192],[251,0.0242],[252,0.0344],[253,0.0459],[254,0.0558],[255,0.0621],[256,0.0617],[257,0.0547],[258,0.0444],[259,0.0437],[260,0.0365],[261,0.0303],[262,0.0271],[263,0.0336],[264,0.0369],[265,0.0455],[266,0.0519],[267,0.0535],[269,0.0432],[270,0.0425],[271,0.0392],[272,0.0368],[273,0.0357],[274,0.032],[275,0.0285],[276,0.0278],[278,0.0301],[279,0.0308],[281,0.0315],[283,0.0318],[284,0.0318],[285,0.0311],[286,0.0294],[287,0.0291],[288,0.0285],[289,0.0269],[290,0.0261],[291,0.0243],[292,0.0219],[293,0.0209],[294,0.0207],[295,0.0228],[296,0.0246],[297,0.025],[298,0.0294],[299,0.0347],[300,0.0443],[301,0.0457],[302,0.0575],[303,0.0596],[304,0.0674],[305,0.0746],[306,0.0798],[307,0.0841],[308,0.0871],[309,0.0895],[310,0.0903],[311,0.0895],[312,0.0893],[313,0.0885],[315,0.088],[316,0.0883],[318,0.0901],[319,0.0911],[320,0.0932],[321,0.0943],[322,0.0944],[324,0.0918],[325,0.0907],[326,0.0906],[329,0.0929],[331,0.0946],[332,0.0963],[333,0.0972],[334,0.0974],[335,0.0988],[336,0.0996],[337,0.1009],[338,0.1027],[340,0.1045],[341,0.1044],[343,0.1056],[344,0.1052],[345,0.1057],[346,0.105],[347,0.1033],[348,0.1036],[349,0.1033],[350,0.104],[351,0.1055],[352,0.1044],[353,0.1049],[354,0.1042],[355,0.1038],[356,0.1038],[357,0.1024],[358,0.1004],[359,0.0978],[360,0.0978]]
       }
       CTRL_L_mouth_cornerDepress {
        type=MovieSceneScriptingFloatChannel
        default=0
        thinned_from=492
        keys [[-131,0.0118],[-130,0.0142],[-129,0.0147],[-128,0.0134],[-127,0.0124],[-126,0.0118],[-125,0.0107],[-124,0.0089],[-123,0.0081],[-121,0.0077],[-120,0.0079],[-119,0.0079],[-118,0.0075],[-116,0.0074],[-115,0.0077],[-113,0.0092],[-112,0.0104],[-111,0.0119],[-110,0.0125],[-109,0.0128],[-108,0.0158],[-107,0.0204],[-106,0.021],[-105,0.0219],[-104,0.0234],[-103,0.0237],[-102,0.0264],[-101,0.0282],[-100,0.0303],[-99,0.0316],[-98,0.034],[-97,0.0382],[-96,0.0409],[-95,0.0415],[-94,0.0453],[-93,0.0452],[-92,0.0424],[-91,0.0356],[-90,0.0282],[-89,0.0196],[-88,0.0126],[-87,0.0086],[-86,0.0078],[-85,0.008],[-84,0.0079],[-81,0.0069],[-79,0.0069],[-78,0.0071],[-77,0.0083],[-76,0.0098],[-75,0.0135],[-74,0.0178],[-73,0.0204],[-71,0.0221],[-70,0.0226],[-69,0.0234],[-68,0.0238],[-67,0.0219],[-66,0.0203],[-65,0.0184],[-64,0.0192],[-63,0.0181],[-61,0.0141],[-60,0.0134],[-59,0.0145],[-57,0.0122],[-56,0.0098],[-54,0.0066],[-53,0.006],[-52,0.0059],[-51,0.0061],[-50,0.0057],[-49,0.0057],[-48,0.005],[-47,0.0048],[-44,0.0048],[-40,0.0059],[-38,0.006],[-36,0.0106],[-35,0.012],[-34,0.0122],[-32,0.0093],[-31,0.0088],[-30,0.0062],[-29,0.0058],[-28,0.0049],[-27,0.0045],[-24,0.0042],[-21,0.0054],[-20,0.0054],[-16,0.0053],[-15,0.005],[-6,0.0054],[-4,0.0052],[-2,0.0055],[1,0.0052],[8,0.0051],[11,0.0048],[19,0.0051],[21,0.0048],[22,0.0044],[28,0.0043],[36,0.0041],[45,0.0046],[49,0.0044],[50,0.0046],[52,0.0057],[54,0.0052],[55,0.0044],[58,0.0023],[59,0.002],[60,0.002],[62,0.0026],[63,0.0026],[64,0.003],[65,0.003],[67,0.0033],[71,0.0031],[73,0.0026],[78,0.0027],[79,0.0024],[83,0.0022],[84,0.0019],[86,0.002],[87,0.0023],[90,0.0019],[97,0.0017],[98,0.0016],[101,0.0024],[102,0.0024],[103,0.0017],[104,0.0014],[105,0.0014],[108,0.0019],[110,0.0019],[112,0.002],[114,0.0016],[118,0.002],[119,0.002],[120,0.0023],[122,0.0024],[123,0.0021],[124,0.002],[126,0.0013],[127,0.0011],[129,0.0016],[131,0.0015],[132,0.0012],[133,0.0013],[134,0.0023],[135,0.0037],[136,0.0041],[138,0.007],[139,0.0064],[140,0.0049],[141,0.0035],[142,0.003],[143,0.0031],[144,0.0036],[145,0.0035],[146,0.003],[150,0.0028],[151,0.0031],[153,0.0034],[154,0.0033],[155,0.0036],[158,0.0029],[159,0.0032],[160,0.0031],[162,0.0031],[163,0.0034],[166,0.0033],[167,0.0035],[168,0.0032],[171,0.0047],[172,0.0045],[173,0.004],[174,0.0032],[175,0.0029],[178,0.0038],[181,0.0031],[182,0.0035],[183,0.0036],[184,0.0033],[185,0.0032],[186,0.0035],[187,0.0039],[191,0.0043],[192,0.0049],[193,0.0063],[194,0.0081],[195,0.008],[196,0.0082],[197,0.0087],[198,0.0096],[199,0.0098],[200,0.0097],[201,0.0092],[202,0.0186],[203,0.0196],[204,0.0201],[205,0.009],[207,0.0099],[209,0.0102],[211,0.0089],[213,0.0085],[214,0.0087],[216,0.0081],[220,0.0077],[221,0.0075],[223,0.0075],[224,0.0082],[225,0.0085],[226,0.0087],[228,0.0082],[229,0.0085],[231,0.0078],[232,0.0071],[233,0.0073],[234,0.0078],[235,0.0078],[238,0.0093],[240,0.0081],[242,0.0081],[243,0.0088],[244,0.0088],[245,0.0086],[246,0.0089],[247,0.0105],[248,0.0113],[249,0.0114],[250,0.0098],[251,0.0089],[252,0.0074],[253,0.0061],[254,0.0052],[255,0.0049],[256,0.0048],[257,0.0041],[258,0.0046],[259,0.0047],[260,0.0064],[261,0.0076],[262,0.0085],[263,0.0089],[267,0.0086],[268,0.0091],[270,0.0093],[271,0.0092],[272,0.0085],[273,0.0084],[278,0.0069],[280,0.0072],[282,0.0068],[285,0.0072],[286,0.0066],[287,0.0069],[288,0.0075],[289,0.0091],[290,0.0096],[293,0.0101],[295,0.0101],[296,0.0108],[297,0.0109],[298,0.0106],[299,0.0098],[300,0.0082],[301,0.008],[302,0.0065],[303,0.0062],[304,0.005],[305,0.0045],[306,0.0042],[311,0.0034],[319,0.0029],[321,0.0033],[323,0.0029],[324,0.0029],[326,0.0037],[329,0.0036],[332,0.004],[340,0.0039],[341,0.0037],[345,0.0037],[346,0.0036],[347,0.0037],[350,0.0033],[352,0.0033],[353,0.0036],[358,0.0035],[360,0.0033]]
       }
       CTRL_L_mouth_stretch {
        type=MovieSceneScriptingFloatChannel
        default=0
        thinned_from=492
        keys [[-131,0.0516],[-130,0.0526],[-129,0.0502],[-128,0.0488],[-127,0.05],[-126,0.0516],[-124,0.0517],[-123,0.0509],[-122,0.0506],[-121,0.0525],[-119,0.0552],[-118,0.0561],[-117,0.0564],[-116,0.0585],[-115,0.0601],[-114,0.0608],[-113,0.0595],[-112,0.0593],[-111,0.0604],[-110,0.0613],[-109,0.0597],[-108,0.0569],[-107,0.0561],[-106,0.0561],[-105,0.0563],[-104,0.0589],[-103,0.061],[-102,0.0635],[-101,0.0645],[-100,0.0678],[-99,0.0654],[-98,0.0642],[-97,0.0654],[-96,0.068],[-95,0.0686],[-94,0.0665],[-92,0.0662],[-90,0.0638],[-89,0.0613],[-88,0.0604],[-87,0.0632],[-86,0.0704],[-85,0.08],[-84,0.0936],[-83,0.106],[-82,0.123],[-81,0.1388],[-80,0.164],[-79,0.1663],[-78,0.1871],[-77,0.205],[-76,0.2169],[-75,0.2257],[-74,0.2369],[-73,0.2442],[-72,0.2499],[-71,0.2562],[-70,0.2509],[-69,0.2529],[-68,0.2514],[-67,0.2487],[-66,0.2503],[-65,0.2555],[-64,0.264],[-62,0.257],[-61,0.2522],[-60,0.243],[-59,0.2218],[-58,0.1764],[-57,0.1651],[-56,0.139],[-55,0.1201],[-54,0.1083],[-53,0.1029],[-52,0.1006],[-51,0.0962],[-50,0.091],[-49,0.0906],[-48,0.0787],[-47,0.0794],[-46,0.0807],[-45,0.087],[-44,0.0873],[-43,0.0885],[-42,0.0951],[-41,0.1055],[-40,0.1223],[-39,0.1377],[-38,0.1415],[-37,0.1744],[-36,0.1911],[-35,0.2023],[-34,0.2192],[-33,0.2234],[-32,0.2135],[-31,0.2025],[-30,0.1367],[-29,0.1253],[-28,0.101],[-27,0.0912],[-26,0.0853],[-25,0.078],[-24,0.0756],[-23,0.0723],[-22,0.0666],[-21,0.065],[-20,0.0642],[-19,0.0693],[-18,0.0686],[-17,0.0677],[-16,0.0624],[-15,0.058],[-14,0.0507],[-13,0.0491],[-12,0.0482],[-11,0.0479],[-10,0.0503],[-8,0.0495],[-7,0.0471],[-6,0.0489],[-5,0.0465],[-4,0.0437],[-3,0.0419],[-2,0.0432],[-1,0.0419],[0,0.0397],[1,0.0391],[2,0.0392],[3,0.0407],[4,0.0398],[5,0.0405],[6,0.0387],[8,0.0379],[9,0.037],[11,0.0367],[12,0.0383],[13,0.039],[14,0.0391],[15,0.0383],[16,0.0381],[19,0.0411],[20,0.0397],[21,0.0386],[22,0.0369],[24,0.0362],[25,0.0366],[26,0.0378],[27,0.0388],[28,0.0389],[30,0.0383],[31,0.0398],[32,0.0405],[33,0.0414],[34,0.0401],[35,0.04],[37,0.0385],[38,0.0383],[40,0.0395],[41,0.0403],[42,0.0396],[43,0.0391],[44,0.0397],[45,0.0398],[47,0.0404],[48,0.0434],[49,0.048],[50,0.0521],[51,0.0588],[52,0.0638],[53,0.078],[54,0.0785],[55,0.0768],[56,0.074],[57,0.0699],[58,0.0676],[59,0.0644],[60,0.0637],[61,0.0642],[62,0.0602],[63,0.0601],[64,0.0574],[65,0.0574],[66,0.0568],[67,0.0592],[68,0.061],[69,0.0615],[70,0.0568],[72,0.0492],[73,0.051],[74,0.0535],[75,0.0556],[76,0.0549],[77,0.0521],[78,0.052],[79,0.0504],[80,0.0494],[81,0.0475],[82,0.0428],[83,0.0393],[85,0.0425],[86,0.043],[87,0.0424],[89,0.0423],[90,0.0409],[91,0.04],[92,0.0412],[94,0.0412],[96,0.0396],[97,0.0403],[98,0.0416],[99,0.0425],[100,0.0423],[101,0.0418],[102,0.0416],[103,0.042],[104,0.0413],[105,0.0413],[106,0.0406],[107,0.0386],[108,0.0382],[109,0.0382],[110,0.0391],[111,0.0413],[113,0.0438],[114,0.0428],[115,0.0425],[117,0.0467],[118,0.0486],[119,0.05],[120,0.0505],[121,0.0496],[122,0.0479],[123,0.0452],[124,0.0434],[125,0.0422],[127,0.0428],[128,0.042],[129,0.0446],[130,0.0447],[131,0.046],[132,0.0434],[133,0.0431],[134,0.0434],[135,0.0493],[136,0.0519],[137,0.0523],[138,0.0459],[139,0.0391],[140,0.0331],[142,0.0282],[143,0.0279],[144,0.0273],[145,0.0275],[146,0.0261],[147,0.0258],[148,0.0252],[149,0.0273],[150,0.0285],[151,0.0284],[153,0.027],[155,0.0277],[158,0.0282],[159,0.0289],[161,0.0276],[162,0.0278],[163,0.0282],[164,0.0278],[165,0.0275],[167,0.0288],[168,0.028],[169,0.0298],[170,0.0337],[171,0.0383],[172,0.0407],[173,0.0398],[174,0.0373],[175,0.0371],[176,0.0393],[177,0.0396],[178,0.0403],[180,0.0385],[181,0.038],[182,0.0363],[183,0.0333],[184,0.0322],[185,0.0333],[186,0.0336],[187,0.0326],[188,0.0325],[189,0.0316],[190,0.034],[191,0.0374],[192,0.0443],[193,0.0489],[194,0.0525],[195,0.0529],[196,0.0574],[197,0.0635],[198,0.0703],[199,0.0748],[200,0.0756],[201,0.0725],[202,0.0703],[203,0.064],[204,0.065],[205,0.0642],[206,0.0642],[208,0.0615],[209,0.06],[210,0.056],[211,0.0512],[212,0.0504],[213,0.0501],[214,0.0508],[215,0.0503],[216,0.0496],[217,0.0485],[218,0.0467],[219,0.0459],[220,0.0458],[221,0.0465],[222,0.0455],[223,0.046],[224,0.0461],[225,0.0479],[226,0.0483],[227,0.0479],[228,0.0502],[229,0.0533],[230,0.0577],[231,0.0601],[232,0.0594],[233,0.0614],[234,0.0638],[235,0.0685],[236,0.0748],[237,0.0761],[238,0.0727],[239,0.0663],[240,0.0669],[241,0.0698],[242,0.0702],[243,0.0735],[244,0.0688],[245,0.0653],[246,0.0663],[247,0.0745],[248,0.0791],[249,0.0862],[250,0.0842],[251,0.0827],[252,0.0728],[253,0.0669],[254,0.0617],[255,0.0597],[257,0.0509],[258,0.0554],[259,0.0559],[260,0.0599],[261,0.0612],[262,0.0594],[263,0.0522],[264,0.0512],[265,0.0499],[266,0.0496],[267,0.0483],[268,0.0477],[270,0.0476],[271,0.0507],[272,0.0529],[273,0.0547],[274,0.0556],[275,0.0557],[277,0.0547],[278,0.0562],[279,0.0582],[280,0.059],[281,0.0575],[282,0.0577],[283,0.0592],[284,0.0595],[285,0.0606],[287,0.0617],[288,0.0619],[290,0.0629],[291,0.0625],[292,0.0604],[293,0.0569],[294,0.0552],[296,0.0534],[297,0.0534],[298,0.0547],[299,0.0545],[300,0.0506],[301,0.0502],[302,0.0482],[303,0.048],[305,0.0465],[306,0.0446],[307,0.0432],[308,0.0422],[310,0.0422],[311,0.0387],[312,0.038],[313,0.0356],[314,0.0328],[315,0.0334],[316,0.0343],[318,0.0348],[319,0.0348],[320,0.0358],[321,0.0359],[322,0.0353],[324,0.0335],[327,0.0328],[328,0.0321],[329,0.0324],[331,0.0352],[332,0.035],[333,0.0339],[334,0.0337],[336,0.0342],[337,0.0342],[338,0.034],[339,0.0353],[340,0.0351],[341,0.0356],[342,0.0341],[343,0.0341],[344,0.0348],[345,0.0352],[346,0.0349],[347,0.0339],[348,0.0342],[349,0.0341],[350,0.0342],[351,0.0339],[352,0.034],[353,0.0335],[356,0.033],[359,0.0299],[360,0.0299]]
       }
       CTRL_L_mouth_lowerLipDepress {
        type=MovieSceneScriptingFloatChannel
        default=0
        thinned_from=492
        keys [[-131,0.0018],[-130,0.0011],[-129,0.0009],[-128,0.0003],[-127,0],[-126,0],[-125,0.0003],[-124,0.0004],[-121,0],[-120,0.0006],[-119,0.0002],[-113,0.0007],[-111,0.001],[-110,0.0004],[-109,0.0003],[-107,0.001],[-106,0.0011],[-105,0.0009],[-104,0.001],[-103,0.0006],[-101,0.0009],[-100,0.0003],[-98,0.001],[-95,0.0011],[-91,0.0005],[-90,0.001],[-88,0.0011],[-87,0.0019],[-86,0.0023],[-82,0.002],[-79,0.0022],[-78,0.0026],[-77,0.0048],[-76,0.0112],[-75,0.0393],[-74,0.0426],[-72,0.0583],[-71,0.0571],[-70,0.0638],[-69,0.0789],[-68,0.0729],[-67,0.0709],[-66,0.0782],[-65,0.085],[-64,0.0795],[-63,0.0745],[-62,0.0809],[-61,0.0811],[-60,0.0718],[-59,0.0495],[-58,0.0239],[-57,0.0204],[-56,0.013],[-55,0.0095],[-54,0.0075],[-53,0.0067],[-52,0.0063],[-50,0.004],[-49,0.0039],[-48,0.0027],[-44,0.0024],[-42,0.0018],[-40,0.0016],[-38,0.0018],[-37,0.0213],[-36,0.0321],[-34,0.0521],[-33,0.0516],[-32,0.0214],[-31,0.0183],[-30,0.0059],[-29,0.0053],[-28,0.0043],[-27,0.0042],[-25,0.0047],[-24,0.0048],[-23,0.004],[-22,0.0029],[-21,0.0029],[-19,0.0025],[-18,0.0028],[-17,0.0024],[-16,0.0026],[-14,0.0025],[-12,0.0028],[-8,0.0028],[-7,0.0026],[-6,0.0026],[-5,0.0025],[-4,0.0026],[-3,0.003],[-2,0.0026],[2,0.0028],[3,0.0026],[4,0.0028],[6,0.0025],[15,0.0027],[17,0.0025],[19,0.0025],[20,0.0028],[26,0.0026],[28,0.0029],[30,0.0025],[36,0.0024],[37,0.0028],[41,0.0029],[42,0.0028],[46,0.0027],[47,0.0025],[50,0.0033],[52,0.003],[54,0.0034],[55,0.0032],[56,0.0027],[57,0.0015],[58,0.0007],[59,0.0015],[60,0.0018],[61,0.0014],[62,0.0012],[63,0.0012],[64,0.0015],[65,0.0015],[66,0.0018],[67,0.0013],[68,0.0011],[69,0.0013],[72,0.0012],[73,0.0008],[75,0.0014],[76,0.0012],[77,0.0014],[80,0.0011],[81,0.0008],[83,0.0011],[84,0.001],[87,0.0014],[88,0.0008],[90,0.0007],[91,0.0005],[92,0.0009],[93,0.0008],[95,0.0003],[97,0.0005],[98,0.0004],[99,0.0009],[100,0.0011],[103,0.0004],[105,0.0005],[106,0.0003],[107,0.0004],[108,0.0001],[112,0.0005],[114,0.0003],[115,0.0007],[117,0.0002],[118,0.0007],[119,0.0004],[121,0.0007],[123,0.0004],[124,0.0006],[125,0.0003],[127,0.0004],[128,0.0001],[129,0.0004],[132,0],[133,0.0002],[134,0.001],[136,0.0009],[137,0.0016],[138,0.0019],[139,0.0006],[140,0.0013],[141,0.001],[142,0.0015],[143,0.0017],[146,0.0009],[147,0.0011],[149,0.0008],[151,0.0011],[152,0.001],[155,0.0014],[156,0.001],[158,0.001],[159,0.0013],[160,0.0009],[162,0.0015],[163,0.0013],[164,0.0009],[165,0.0011],[166,0.001],[167,0.0013],[168,0.0013],[169,0.0021],[170,0.0022],[171,0.0026],[172,0.0023],[173,0.0013],[174,0.0006],[177,0.0012],[178,0.0012],[179,0.0004],[180,0.0003],[181,0.0008],[182,0.0008],[184,0.0003],[188,0.0005],[190,0.0001],[191,0.0003],[192,0.0003],[194,0.0023],[195,0.0024],[196,0.002],[197,0.0011],[199,0.0019],[200,0.0019],[201,0.0016],[202,0.0024],[203,0.0019],[207,0.0017],[209,0.0019],[210,0.0018],[212,0.0018],[214,0.0015],[215,0.0017],[217,0.0016],[218,0.0017],[219,0.0015],[220,0.0019],[221,0.0018],[223,0.0019],[224,0.0018],[227,0.0022],[229,0.0016],[231,0.0017],[234,0.0015],[235,0.0005],[236,0.0017],[237,0.002],[238,0.0019],[239,0.0009],[240,0.0008],[241,0.001],[242,0.0011],[243,0.0004],[244,0.0001],[245,0.0005],[246,0.0002],[247,0.0002],[248,0],[249,0.0004],[250,0.0001],[251,0.0003],[252,0],[255,0.0002],[256,0],[259,0],[261,0.0014],[262,0.0014],[263,0.0023],[266,0.0027],[268,0.0023],[269,0.0018],[271,0.0017],[274,0.0004],[277,0],[281,0.0004],[282,0.0002],[285,0.0004],[286,0],[288,0.0001],[292,0],[293,0.0001],[294,0],[295,0.0004],[298,0.0007],[300,0.0012],[304,0.0013],[307,0.0021],[309,0.0015],[318,0.001],[321,0.0014],[322,0.001],[323,0.0008],[324,0.0014],[326,0.0015],[327,0.0013],[332,0.0017],[334,0.0015],[336,0.0017],[337,0.0014],[338,0.0016],[339,0.0014],[344,0.0015],[347,0.0012],[348,0.0014],[349,0.0011],[352,0.0012],[353,0.0014],[358,0.0015],[360,0.0013]]
       }
       CTRL_R_mouth_lowerLipDepress {
        type=MovieSceneScriptingFloatChannel
        default=0
        thinned_from=492
        keys [[-131,0.003],[-130,0.0017],[-129,0.0014],[-128,0.0007],[-127,0.0003],[-126,0.0002],[-124,0.0006],[-123,0.0003],[-121,0],[-120,0.0006],[-119,0.0002],[-118,0.0005],[-117,0.0006],[-116,0.0011],[-114,0.0011],[-113,0.001],[-111,0.0014],[-110,0.001],[-109,0.0009],[-108,0.0015],[-107,0.0018],[-104,0.0018],[-103,0.0013],[-101,0.0011],[-100,0.0005],[-98,0.0011],[-95,0.0011],[-94,0.0013],[-93,0.0013],[-91,0.0008],[-90,0.0014],[-89,0.0016],[-88,0.0016],[-87,0.0022],[-86,0.0025],[-79,0.003],[-78,0.0038],[-77,0.0107],[-76,0.0304],[-75,0.0802],[-74,0.0972],[-73,0.1104],[-72,0.1342],[-71,0.1374],[-70,0.1517],[-69,0.1766],[-68,0.1687],[-67,0.1683],[-66,0.1786],[-65,0.1894],[-64,0.1841],[-63,0.1766],[-62,0.1837],[-61,0.1774],[-60,0.1533],[-59,0.1196],[-58,0.0914],[-57,0.0841],[-56,0.0691],[-55,0.0611],[-54,0.0572],[-53,0.053],[-52,0.0493],[-51,0.0299],[-50,0.0202],[-49,0.0192],[-48,0.0069],[-47,0.0049],[-46,0.0047],[-45,0.0039],[-44,0.0037],[-43,0.0031],[-42,0.0028],[-38,0.0028],[-37,0.052],[-36,0.0814],[-35,0.1065],[-34,0.1236],[-33,0.1167],[-32,0.0615],[-31,0.0543],[-30,0.0248],[-29,0.0225],[-28,0.0193],[-27,0.0212],[-26,0.0204],[-25,0.0236],[-24,0.0251],[-23,0.0237],[-22,0.0179],[-21,0.009],[-20,0.0065],[-19,0.0044],[-17,0.0038],[-16,0.0039],[-15,0.0036],[-10,0.004],[-6,0.0041],[-4,0.0039],[-3,0.0041],[-2,0.004],[-1,0.0041],[1,0.0039],[2,0.0041],[5,0.0039],[6,0.0037],[8,0.0038],[10,0.0036],[15,0.0038],[17,0.0035],[20,0.0034],[21,0.0036],[22,0.0033],[27,0.0036],[29,0.0036],[30,0.0037],[33,0.0035],[34,0.0038],[35,0.0037],[36,0.0034],[37,0.0037],[39,0.0038],[42,0.0036],[46,0.0039],[47,0.0038],[51,0.0039],[52,0.0047],[53,0.0114],[54,0.0113],[55,0.0082],[56,0.005],[57,0.003],[58,0.0018],[62,0.0025],[65,0.0023],[66,0.0027],[67,0.0022],[68,0.0022],[69,0.0025],[71,0.0023],[72,0.0024],[73,0.0021],[74,0.0026],[75,0.004],[76,0.0036],[77,0.0065],[78,0.0066],[79,0.0073],[80,0.0073],[81,0.0038],[82,0.0046],[83,0.007],[84,0.0072],[85,0.0081],[86,0.0061],[87,0.007],[88,0.0039],[90,0.0031],[91,0.0029],[92,0.0067],[93,0.0064],[94,0.0051],[95,0.0028],[96,0.002],[98,0.0014],[99,0.002],[100,0.0022],[101,0.0018],[102,0.0017],[103,0.0013],[107,0.0014],[108,0.0011],[111,0.0016],[112,0.0015],[114,0.0009],[115,0.0016],[116,0.0017],[117,0.0012],[118,0.0014],[119,0.0011],[121,0.0015],[122,0.0019],[123,0.0019],[124,0.0013],[126,0.0057],[127,0.006],[128,0.005],[129,0.0073],[130,0.0099],[131,0.0111],[132,0.0084],[133,0.0099],[134,0.0061],[135,0.0033],[136,0.0021],[137,0.0027],[138,0.0038],[139,0.0014],[140,0.0018],[141,0.0013],[143,0.0019],[144,0.0018],[146,0.001],[147,0.0013],[148,0.0009],[149,0.0008],[151,0.0013],[152,0.0013],[153,0.0016],[154,0.0013],[155,0.0017],[156,0.001],[158,0.0012],[159,0.0019],[160,0.0013],[161,0.0015],[162,0.0021],[163,0.0019],[164,0.0013],[165,0.0015],[166,0.0015],[167,0.0018],[168,0.0014],[169,0.0023],[170,0.0025],[171,0.0028],[172,0.0025],[174,0.0011],[175,0.0013],[176,0.0019],[178,0.0021],[179,0.0011],[180,0.001],[181,0.0013],[182,0.0012],[184,0.0006],[185,0.001],[188,0.0012],[190,0.0009],[191,0.0011],[192,0.0008],[194,0.0034],[195,0.0033],[197,0.0017],[199,0.0029],[200,0.0029],[201,0.0023],[202,0.0032],[203,0.003],[206,0.0033],[207,0.0031],[211,0.0033],[213,0.0029],[216,0.0029],[217,0.0026],[220,0.0031],[222,0.003],[223,0.0034],[224,0.0031],[227,0.0033],[234,0.0025],[235,0.0013],[236,0.0024],[237,0.0028],[238,0.0028],[239,0.0016],[240,0.0013],[241,0.0018],[242,0.002],[244,0.0011],[245,0.0013],[246,0.0008],[247,0.0008],[248,0.0003],[249,0.0011],[250,0.0003],[251,0.0006],[252,0.0002],[253,0],[255,0.0002],[256,0],[258,0],[259,0.0002],[260,0.0016],[261,0.0023],[262,0.002],[263,0.0029],[265,0.0032],[267,0.0029],[268,0.003],[269,0.0026],[271,0.0023],[272,0.0019],[273,0.0012],[274,0.0007],[278,0.0003],[279,0.0004],[281,0.0007],[282,0.0006],[284,0.0008],[285,0.0006],[286,0],[288,0.0005],[291,0.0002],[292,0.0002],[293,0.0005],[294,0.0005],[296,0.0014],[298,0.0015],[300,0.0027],[307,0.0036],[310,0.0034],[311,0.003],[313,0.003],[315,0.0027],[316,0.003],[318,0.0026],[320,0.003],[321,0.003],[322,0.0025],[323,0.0023],[324,0.0029],[325,0.0028],[326,0.0029],[329,0.0027],[332,0.0031],[333,0.0028],[335,0.0029],[337,0.0027],[338,0.003],[339,0.0027],[341,0.0023],[343,0.0026],[347,0.0024],[348,0.0026],[350,0.0022],[351,0.0025],[352,0.0025],[353,0.0028],[354,0.0025],[360,0.0024]]
       }
       CTRL_R_mouth_stretch {
        type=MovieSceneScriptingFloatChannel
        default=0
        thinned_from=492
        keys [[-131,0.0515],[-130,0.0515],[-129,0.0492],[-128,0.0478],[-127,0.0485],[-126,0.0497],[-124,0.0485],[-123,0.0474],[-122,0.0471],[-121,0.0492],[-120,0.05],[-119,0.0512],[-118,0.0527],[-117,0.053],[-116,0.0551],[-115,0.0566],[-114,0.0565],[-113,0.0546],[-112,0.0536],[-110,0.0561],[-109,0.0546],[-108,0.0522],[-106,0.0523],[-105,0.0532],[-104,0.0552],[-103,0.0567],[-102,0.0588],[-101,0.0598],[-100,0.0633],[-99,0.0621],[-98,0.0616],[-97,0.0631],[-96,0.066],[-95,0.0667],[-94,0.0641],[-92,0.0639],[-91,0.0643],[-90,0.0642],[-89,0.0629],[-88,0.0638],[-87,0.0694],[-86,0.079],[-85,0.0911],[-84,0.1075],[-83,0.1223],[-82,0.1411],[-81,0.1581],[-80,0.1826],[-79,0.1846],[-78,0.2017],[-77,0.2119],[-76,0.2176],[-75,0.2199],[-74,0.2271],[-72,0.2369],[-71,0.2411],[-70,0.2351],[-69,0.238],[-68,0.238],[-67,0.2365],[-66,0.2384],[-65,0.244],[-64,0.2528],[-63,0.2488],[-62,0.2453],[-61,0.241],[-60,0.2328],[-59,0.2144],[-58,0.1786],[-57,0.1688],[-56,0.1446],[-55,0.1233],[-54,0.1094],[-53,0.1034],[-52,0.1008],[-50,0.0882],[-49,0.0877],[-48,0.0742],[-47,0.0749],[-46,0.0769],[-45,0.0866],[-44,0.0875],[-43,0.0907],[-42,0.0986],[-41,0.1106],[-40,0.1281],[-39,0.1425],[-38,0.1457],[-37,0.1723],[-36,0.1848],[-35,0.1938],[-34,0.2115],[-33,0.219],[-32,0.2118],[-31,0.2014],[-30,0.1396],[-29,0.128],[-28,0.1033],[-27,0.0926],[-26,0.0856],[-25,0.0777],[-24,0.0751],[-23,0.072],[-22,0.0648],[-21,0.0609],[-20,0.0605],[-19,0.0684],[-18,0.0721],[-17,0.071],[-16,0.0654],[-15,0.0596],[-14,0.0513],[-13,0.0487],[-12,0.0473],[-11,0.0467],[-10,0.0493],[-8,0.0498],[-7,0.0486],[-6,0.05],[-5,0.047],[-4,0.0431],[-3,0.0413],[-2,0.0433],[-1,0.0422],[0,0.0404],[1,0.0395],[2,0.0397],[3,0.0413],[4,0.0412],[5,0.0421],[6,0.0398],[8,0.0385],[9,0.037],[10,0.036],[11,0.0357],[12,0.0378],[13,0.0386],[14,0.0387],[15,0.0383],[16,0.0387],[17,0.0397],[18,0.0404],[19,0.0412],[21,0.0394],[22,0.0377],[23,0.0374],[24,0.0367],[25,0.0372],[27,0.0398],[28,0.04],[30,0.0398],[31,0.0411],[32,0.0414],[33,0.0423],[34,0.0408],[35,0.0407],[37,0.0392],[38,0.039],[39,0.0392],[40,0.0391],[41,0.0392],[42,0.0382],[43,0.0378],[44,0.0393],[45,0.0396],[46,0.0403],[47,0.0394],[48,0.041],[50,0.0503],[51,0.0577],[52,0.0629],[53,0.077],[54,0.0771],[55,0.075],[56,0.0721],[57,0.0684],[58,0.0664],[59,0.0634],[60,0.0627],[61,0.0634],[62,0.0598],[63,0.0596],[64,0.0557],[65,0.0557],[66,0.0551],[67,0.0581],[68,0.0616],[69,0.0627],[71,0.0539],[72,0.0502],[73,0.0522],[74,0.0547],[75,0.0569],[76,0.0561],[77,0.053],[78,0.0528],[79,0.0508],[80,0.0502],[81,0.0472],[82,0.0419],[83,0.0378],[84,0.0401],[85,0.0413],[86,0.041],[87,0.0398],[88,0.04],[89,0.0404],[90,0.0391],[91,0.0384],[92,0.0399],[94,0.0403],[96,0.0392],[97,0.04],[99,0.0421],[100,0.0423],[102,0.042],[103,0.0431],[105,0.0431],[106,0.0421],[107,0.0401],[108,0.0393],[109,0.0396],[110,0.0403],[111,0.0425],[113,0.0451],[114,0.0444],[115,0.0439],[116,0.0465],[117,0.0474],[118,0.0481],[120,0.0471],[121,0.046],[122,0.0444],[123,0.0415],[124,0.0398],[125,0.0393],[127,0.0408],[128,0.0395],[129,0.0422],[130,0.043],[131,0.0453],[132,0.0423],[134,0.041],[135,0.0461],[136,0.0464],[137,0.0435],[138,0.0349],[139,0.0287],[140,0.0255],[141,0.0248],[142,0.0244],[143,0.0233],[144,0.0215],[145,0.0219],[146,0.021],[147,0.0208],[148,0.0199],[149,0.0219],[150,0.0234],[151,0.0235],[152,0.0234],[153,0.0226],[154,0.0227],[156,0.022],[158,0.0222],[159,0.0237],[160,0.0237],[161,0.0227],[163,0.0226],[164,0.0222],[165,0.0216],[166,0.0227],[167,0.0229],[168,0.0216],[169,0.0228],[170,0.0272],[171,0.0333],[172,0.0372],[173,0.0365],[174,0.034],[175,0.0336],[176,0.0366],[177,0.037],[178,0.0379],[179,0.0378],[180,0.0373],[181,0.0386],[182,0.0377],[183,0.0348],[184,0.0337],[185,0.0347],[186,0.0349],[187,0.0341],[188,0.0339],[189,0.0332],[191,0.0381],[192,0.0442],[193,0.0479],[194,0.0514],[195,0.0525],[196,0.0578],[197,0.0637],[198,0.0688],[199,0.0717],[200,0.0722],[201,0.0697],[202,0.0684],[203,0.0622],[204,0.0617],[205,0.0594],[206,0.0584],[207,0.0567],[208,0.0553],[209,0.0542],[210,0.0507],[211,0.0466],[212,0.0459],[213,0.0462],[214,0.047],[215,0.0471],[216,0.0466],[217,0.0468],[218,0.0454],[219,0.0446],[220,0.0436],[221,0.0439],[222,0.0431],[223,0.0434],[224,0.0432],[225,0.0445],[226,0.0447],[227,0.0442],[228,0.0473],[229,0.0509],[230,0.0555],[231,0.0573],[232,0.0563],[233,0.0575],[234,0.059],[235,0.062],[236,0.0682],[237,0.0707],[238,0.0688],[239,0.0641],[241,0.0658],[242,0.0657],[243,0.066],[244,0.0612],[245,0.0595],[246,0.063],[247,0.0725],[248,0.0787],[249,0.0877],[250,0.087],[251,0.086],[252,0.0768],[254,0.0675],[255,0.0654],[256,0.0609],[257,0.0562],[258,0.0596],[259,0.06],[260,0.0632],[261,0.0631],[262,0.0607],[263,0.052],[264,0.051],[265,0.0505],[266,0.0512],[267,0.0499],[268,0.0489],[270,0.0491],[271,0.0522],[272,0.0537],[274,0.0555],[275,0.0558],[276,0.0553],[277,0.0549],[278,0.0565],[279,0.0578],[280,0.0578],[281,0.0565],[282,0.0573],[283,0.059],[284,0.0593],[286,0.0614],[287,0.062],[288,0.0624],[290,0.0637],[291,0.0638],[292,0.0619],[293,0.0576],[294,0.055],[295,0.0545],[296,0.0558],[297,0.0561],[298,0.0599],[299,0.0603],[300,0.0541],[301,0.0534],[302,0.0491],[303,0.0487],[305,0.0466],[306,0.0447],[308,0.0428],[310,0.0434],[311,0.0406],[312,0.0399],[314,0.0339],[315,0.0341],[316,0.0352],[317,0.0352],[318,0.0361],[319,0.0361],[320,0.0371],[321,0.0372],[322,0.036],[323,0.0352],[324,0.0335],[327,0.0328],[328,0.0314],[329,0.0317],[330,0.0329],[331,0.0343],[332,0.0337],[333,0.0323],[334,0.0316],[335,0.0316],[338,0.0326],[339,0.0337],[340,0.0335],[341,0.0338],[342,0.033],[343,0.0333],[344,0.034],[345,0.0343],[346,0.034],[347,0.0331],[348,0.0339],[349,0.0341],[350,0.0346],[351,0.0345],[352,0.035],[354,0.0347],[356,0.0343],[358,0.0318],[359,0.0312],[360,0.0312]]
       }
       CTRL_L_mouth_towardsD {
        type=MovieSceneScriptingFloatChannel
        default=0
        thinned_from=492
        keys [[-131,0.0034],[-130,0.0023],[-129,0.0022],[-127,0.0013],[-126,0.0011],[-124,0.0017],[-123,0.0012],[-121,0.001],[-120,0.0017],[-119,0.0016],[-117,0.0019],[-115,0.0023],[-113,0.002],[-111,0.0021],[-110,0.0015],[-109,0.0014],[-107,0.0027],[-106,0.0027],[-104,0.0026],[-103,0.0022],[-101,0.0024],[-100,0.0018],[-98,0.0023],[-93,0.0022],[-92,0.0018],[-91,0.0017],[-88,0.0025],[-86,0.0034],[-85,0.0036],[-83,0.0035],[-80,0.0038],[-79,0.0038],[-77,0.0034],[-76,0.0037],[-74,0.0035],[-73,0.0038],[-71,0.0037],[-68,0.004],[-65,0.0037],[-64,0.0038],[-61,0.0037],[-56,0.0042],[-54,0.0038],[-52,0.0042],[-51,0.004],[-49,0.0041],[-47,0.0038],[-44,0.0036],[-42,0.0031],[-38,0.0035],[-37,0.0039],[-35,0.0041],[-33,0.0039],[-32,0.0042],[-31,0.0042],[-30,0.0038],[-28,0.0038],[-27,0.0039],[-24,0.0035],[-23,0.0032],[-22,0.0032],[-21,0.0034],[-20,0.004],[-19,0.0039],[-17,0.0041],[-15,0.0039],[-14,0.0041],[-12,0.0039],[-8,0.0045],[-7,0.0043],[-6,0.0043],[-5,0.004],[-4,0.004],[-3,0.0041],[-2,0.004],[-1,0.0041],[1,0.0039],[2,0.0041],[3,0.004],[7,0.0044],[13,0.0039],[16,0.0041],[19,0.0041],[20,0.0043],[21,0.0042],[22,0.0038],[28,0.0037],[30,0.0042],[33,0.0038],[42,0.0039],[43,0.0042],[45,0.0042],[46,0.004],[48,0.004],[49,0.0038],[52,0.0041],[53,0.0038],[55,0.004],[56,0.0037],[58,0.0019],[62,0.0028],[63,0.0027],[64,0.0025],[65,0.0025],[66,0.0031],[67,0.0027],[68,0.0027],[69,0.003],[70,0.0026],[72,0.0027],[73,0.0023],[74,0.0025],[75,0.003],[76,0.0027],[78,0.0028],[80,0.0023],[81,0.0024],[82,0.0021],[83,0.0023],[84,0.0018],[87,0.0021],[88,0.0017],[89,0.0018],[90,0.0016],[93,0.0017],[95,0.0015],[97,0.0018],[98,0.0017],[100,0.0023],[102,0.0022],[103,0.0016],[106,0.0016],[107,0.0018],[108,0.0016],[113,0.0021],[114,0.0017],[115,0.0023],[117,0.0018],[119,0.0018],[121,0.0021],[123,0.0018],[124,0.0019],[125,0.0014],[127,0.0018],[128,0.0016],[130,0.0019],[131,0.0015],[132,0.0009],[134,0.0023],[136,0.0021],[138,0.0036],[139,0.0022],[140,0.0033],[141,0.003],[143,0.0034],[144,0.0033],[146,0.0024],[147,0.0027],[148,0.0023],[149,0.0023],[151,0.0028],[152,0.0027],[153,0.0029],[154,0.0028],[155,0.003],[156,0.0027],[158,0.0025],[159,0.0031],[160,0.0027],[162,0.0032],[163,0.0029],[164,0.0028],[165,0.003],[166,0.0028],[167,0.0029],[168,0.0026],[169,0.0034],[171,0.0039],[172,0.0036],[173,0.0027],[174,0.0021],[175,0.0024],[178,0.0028],[179,0.002],[182,0.002],[184,0.0015],[187,0.0021],[189,0.002],[190,0.0016],[191,0.0017],[192,0.0013],[194,0.0037],[195,0.0038],[196,0.0033],[197,0.0026],[199,0.0038],[200,0.0038],[201,0.0034],[202,0.0038],[204,0.0035],[206,0.0036],[207,0.0035],[208,0.0036],[210,0.0034],[211,0.0036],[214,0.0032],[216,0.0034],[217,0.0031],[218,0.0033],[219,0.003],[220,0.0036],[222,0.0032],[223,0.0033],[224,0.0032],[227,0.0034],[231,0.0029],[233,0.0033],[234,0.0031],[235,0.0021],[236,0.0033],[237,0.0035],[238,0.0034],[239,0.0023],[240,0.0022],[242,0.0024],[244,0.0016],[245,0.0018],[246,0.0015],[247,0.0017],[248,0.0011],[249,0.0017],[250,0.001],[251,0.0019],[252,0.0016],[253,0.0015],[254,0.002],[255,0.0021],[256,0.0015],[257,0.0007],[258,0.0002],[259,0.0004],[260,0.0022],[261,0.0028],[262,0.0025],[263,0.0036],[265,0.0038],[268,0.0036],[269,0.0032],[270,0.0031],[271,0.0033],[272,0.0031],[274,0.0018],[278,0.0016],[281,0.002],[282,0.0019],[285,0.0021],[286,0.001],[288,0.0017],[292,0.0015],[294,0.0016],[295,0.0021],[298,0.0022],[300,0.003],[304,0.0028],[307,0.0033],[310,0.0031],[311,0.0029],[314,0.0029],[315,0.0028],[316,0.003],[318,0.0027],[321,0.0031],[322,0.0026],[323,0.0023],[324,0.0027],[326,0.0029],[327,0.0028],[328,0.003],[330,0.003],[331,0.0031],[333,0.0028],[335,0.0029],[337,0.0027],[338,0.003],[341,0.0026],[347,0.0029],[350,0.0026],[351,0.0028],[353,0.0028],[354,0.0026],[360,0.0027]]
       }
       CTRL_R_mouth_funnelU {
        type=MovieSceneScriptingFloatChannel
        default=0
        thinned_from=492
        keys [[-131,0.0025],[-130,0.0014],[-129,0.0012],[-127,0.0003],[-126,0.0002],[-124,0.0005],[-121,0],[-120,0.0004],[-119,0.0002],[-118,0.0005],[-117,0.0006],[-114,0.0015],[-112,0.0014],[-111,0.0014],[-110,0.0008],[-109,0.0007],[-107,0.0019],[-106,0.002],[-104,0.0016],[-103,0.0011],[-102,0.0013],[-101,0.0012],[-100,0.0005],[-98,0.001],[-91,0.0008],[-90,0.0013],[-88,0.0016],[-86,0.0024],[-83,0.0028],[-81,0.0025],[-79,0.0026],[-77,0.0025],[-76,0.0026],[-74,0.0021],[-72,0.0024],[-71,0.0023],[-70,0.0026],[-67,0.0027],[-63,0.0022],[-60,0.0026],[-57,0.0025],[-55,0.0027],[-54,0.0026],[-52,0.003],[-40,0.0026],[-38,0.0028],[-36,0.0024],[-33,0.0025],[-32,0.0028],[-28,0.0024],[-23,0.0022],[-22,0.0025],[-21,0.0025],[-20,0.0028],[-19,0.0027],[-18,0.0032],[-16,0.0027],[-13,0.0029],[-12,0.0028],[-11,0.0029],[-10,0.0028],[-7,0.003],[-6,0.0033],[0,0.0029],[1,0.0031],[3,0.003],[4,0.0031],[5,0.003],[6,0.0033],[9,0.0031],[18,0.0032],[19,0.003],[21,0.0032],[22,0.0029],[25,0.0028],[30,0.0032],[32,0.003],[35,0.0031],[36,0.0029],[37,0.0031],[45,0.0031],[46,0.0033],[48,0.003],[51,0.0026],[52,0.0028],[53,0.0026],[54,0.0028],[56,0.0025],[57,0.0017],[58,0.0007],[59,0.0011],[62,0.0015],[65,0.0016],[66,0.002],[68,0.0016],[69,0.002],[72,0.0017],[73,0.0015],[74,0.0016],[75,0.0019],[76,0.0015],[78,0.0016],[80,0.0013],[81,0.0014],[82,0.0011],[83,0.0012],[84,0.0009],[87,0.0016],[88,0.0009],[89,0.0011],[94,0.0008],[97,0.0009],[98,0.0008],[100,0.0016],[102,0.0012],[103,0.0007],[105,0.0009],[108,0.0008],[109,0.0009],[110,0.0008],[111,0.0011],[113,0.001],[114,0.0005],[116,0.0013],[117,0.0006],[118,0.0009],[119,0.0005],[122,0.0007],[123,0.0004],[124,0.0005],[125,0.0002],[127,0.0002],[128,0],[129,0.0004],[131,0],[132,0],[133,0.0003],[134,0.0012],[136,0.0011],[137,0.0023],[138,0.0099],[139,0.0095],[140,0.0121],[141,0.0122],[142,0.0133],[143,0.0128],[144,0.0117],[145,0.0091],[146,0.0099],[147,0.0097],[148,0.011],[149,0.0074],[151,0.011],[152,0.0107],[153,0.0126],[154,0.0128],[155,0.0097],[156,0.0095],[157,0.0099],[158,0.0089],[159,0.0095],[160,0.0091],[161,0.0096],[163,0.0089],[164,0.007],[165,0.0069],[168,0.0022],[170,0.0023],[171,0.0026],[172,0.0022],[173,0.0014],[174,0.001],[177,0.0015],[178,0.0015],[179,0.0007],[180,0.0007],[181,0.0011],[182,0.0011],[184,0.0006],[185,0.0009],[186,0.0008],[187,0.0012],[188,0.0013],[190,0.0007],[192,0.0007],[193,0.002],[194,0.0029],[195,0.0028],[197,0.0016],[198,0.0022],[199,0.0026],[200,0.0025],[201,0.0022],[202,0.003],[207,0.0025],[211,0.0028],[214,0.0021],[216,0.0024],[217,0.0021],[218,0.0023],[219,0.0022],[220,0.0025],[222,0.0025],[223,0.0028],[224,0.0024],[227,0.0026],[229,0.0022],[230,0.0024],[231,0.0021],[232,0.002],[233,0.0024],[234,0.0021],[235,0.0012],[236,0.0025],[237,0.0029],[238,0.0028],[239,0.0014],[240,0.0014],[241,0.0018],[242,0.0018],[243,0.0011],[244,0.0008],[245,0.0012],[246,0.0008],[247,0.0008],[248,0.0003],[249,0.0008],[250,0.0002],[251,0.0007],[252,0.0002],[253,0],[254,0.0008],[257,0],[259,0.0001],[260,0.0015],[261,0.0022],[262,0.0018],[263,0.0028],[265,0.0032],[268,0.0031],[269,0.0026],[271,0.0023],[272,0.0019],[274,0.0005],[276,0.0003],[278,0.0004],[279,0.0003],[281,0.0006],[282,0.0003],[283,0.0005],[285,0.0006],[286,0.0001],[289,0.0005],[291,0.0002],[292,0.0002],[293,0.0006],[296,0.0013],[298,0.0014],[300,0.0023],[305,0.0022],[307,0.0026],[313,0.0023],[318,0.0018],[321,0.0022],[322,0.0017],[323,0.0016],[324,0.002],[325,0.0021],[327,0.002],[332,0.0024],[333,0.0021],[335,0.0022],[337,0.002],[338,0.0021],[342,0.0017],[343,0.002],[346,0.0018],[347,0.0021],[348,0.0018],[349,0.0018],[350,0.0016],[351,0.002],[352,0.002],[353,0.0022],[357,0.002],[358,0.0021],[360,0.002]]
       }
       CTRL_L_mouth_towardsU {
        type=MovieSceneScriptingFloatChannel
        default=0
        thinned_from=492
        keys [[-131,0.0092],[-130,0.0082],[-129,0.008],[-128,0.0071],[-126,0.0068],[-124,0.0078],[-123,0.0068],[-122,0.0065],[-121,0.0066],[-120,0.0073],[-117,0.0069],[-116,0.0066],[-114,0.0071],[-111,0.0068],[-110,0.0062],[-109,0.006],[-107,0.0065],[-106,0.0065],[-105,0.0063],[-104,0.0066],[-103,0.0062],[-101,0.0069],[-100,0.0058],[-99,0.0064],[-96,0.0074],[-94,0.0076],[-93,0.0075],[-92,0.0064],[-91,0.0064],[-88,0.0076],[-87,0.0083],[-86,0.0081],[-84,0.0066],[-83,0.0063],[-82,0.0064],[-79,0.0065],[-78,0.0061],[-77,0.0053],[-76,0.0055],[-75,0.0049],[-74,0.0046],[-73,0.0051],[-72,0.0053],[-71,0.0049],[-70,0.005],[-69,0.0049],[-67,0.0051],[-65,0.0046],[-62,0.0052],[-61,0.0052],[-60,0.0057],[-59,0.0071],[-58,0.0093],[-57,0.0098],[-56,0.0109],[-53,0.0117],[-52,0.0118],[-51,0.0125],[-48,0.0124],[-47,0.0118],[-46,0.0115],[-45,0.0097],[-44,0.0091],[-43,0.0072],[-42,0.0062],[-41,0.0057],[-40,0.0056],[-38,0.0062],[-37,0.0059],[-35,0.006],[-34,0.0058],[-33,0.0059],[-32,0.008],[-31,0.0084],[-30,0.0099],[-28,0.0103],[-26,0.0104],[-24,0.0107],[-23,0.0143],[-22,0.0167],[-21,0.0162],[-20,0.0151],[-18,0.0103],[-17,0.0103],[-15,0.0095],[-14,0.01],[-12,0.0095],[-11,0.0101],[-10,0.0098],[-9,0.0104],[-8,0.0103],[-7,0.0095],[-4,0.0103],[-2,0.0101],[-1,0.0102],[0,0.0098],[3,0.0101],[4,0.01],[5,0.0091],[7,0.0096],[10,0.01],[13,0.0088],[14,0.009],[15,0.0095],[16,0.0088],[17,0.0084],[18,0.0082],[20,0.0085],[21,0.0084],[22,0.0078],[24,0.0081],[25,0.0084],[28,0.0087],[30,0.0083],[32,0.0086],[33,0.0078],[34,0.0074],[35,0.0075],[36,0.0081],[38,0.0078],[39,0.0079],[42,0.0091],[44,0.0087],[45,0.0088],[46,0.0115],[47,0.0107],[48,0.0119],[49,0.0121],[50,0.0114],[52,0.0118],[54,0.0127],[55,0.0127],[57,0.0106],[58,0.0093],[59,0.0087],[60,0.0086],[61,0.0094],[63,0.0095],[64,0.01],[66,0.0098],[68,0.009],[69,0.0092],[70,0.0088],[72,0.009],[73,0.008],[74,0.0083],[75,0.0088],[76,0.0084],[77,0.0091],[78,0.0091],[79,0.0087],[80,0.0077],[81,0.0083],[82,0.0081],[84,0.0075],[86,0.0082],[87,0.0077],[88,0.0063],[90,0.0068],[91,0.0068],[92,0.0071],[96,0.0062],[98,0.0052],[99,0.0057],[100,0.0067],[102,0.0063],[103,0.0056],[104,0.006],[106,0.0062],[107,0.0066],[108,0.0065],[109,0.0072],[110,0.0069],[111,0.0078],[112,0.008],[113,0.0077],[114,0.0076],[115,0.0087],[116,0.0091],[118,0.0106],[119,0.0106],[120,0.0117],[121,0.0122],[122,0.0121],[123,0.0115],[124,0.0116],[125,0.0113],[126,0.0117],[127,0.0119],[128,0.0113],[129,0.0112],[130,0.012],[131,0.011],[132,0.0111],[133,0.0128],[134,0.0143],[135,0.015],[136,0.0172],[137,0.017],[138,0.0156],[139,0.0105],[141,0.0107],[142,0.0112],[143,0.0125],[144,0.0123],[145,0.0116],[146,0.0112],[147,0.0121],[148,0.0109],[149,0.0104],[150,0.0115],[151,0.012],[152,0.0116],[154,0.0117],[155,0.0122],[156,0.0113],[157,0.011],[158,0.0099],[159,0.0106],[160,0.0095],[162,0.0095],[164,0.0088],[165,0.0092],[166,0.0089],[168,0.0087],[169,0.01],[170,0.0095],[171,0.0093],[172,0.0095],[173,0.0092],[174,0.0079],[177,0.0085],[178,0.0085],[179,0.0069],[180,0.0063],[182,0.0061],[183,0.0053],[184,0.0055],[185,0.0063],[186,0.0063],[187,0.0076],[188,0.0078],[190,0.0075],[191,0.0083],[192,0.0102],[193,0.0147],[194,0.0162],[195,0.0142],[196,0.0088],[197,0.0054],[198,0.0056],[199,0.0063],[200,0.0064],[201,0.0061],[202,0.008],[203,0.0069],[204,0.0065],[205,0.0067],[207,0.0068],[209,0.007],[210,0.0066],[211,0.0071],[213,0.0067],[217,0.0066],[218,0.0069],[219,0.0069],[220,0.0072],[221,0.0071],[222,0.0066],[223,0.0068],[224,0.0065],[227,0.0068],[228,0.0063],[230,0.006],[231,0.0055],[233,0.0057],[234,0.0054],[235,0.0037],[236,0.0056],[237,0.0067],[238,0.0069],[239,0.004],[240,0.0037],[241,0.0046],[242,0.0048],[244,0.0036],[245,0.0039],[246,0.0035],[247,0.0045],[248,0.0054],[249,0.0059],[250,0.005],[251,0.0055],[253,0.0039],[255,0.0058],[256,0.0051],[257,0.0048],[258,0.0057],[259,0.006],[260,0.0089],[261,0.0097],[262,0.0091],[265,0.0088],[266,0.009],[268,0.0108],[269,0.0096],[270,0.0093],[271,0.0081],[272,0.0071],[274,0.0057],[275,0.0058],[277,0.0052],[279,0.005],[281,0.0054],[282,0.0053],[283,0.0056],[285,0.0056],[286,0.0039],[288,0.0044],[291,0.0042],[292,0.0049],[293,0.006],[295,0.0062],[296,0.0066],[297,0.0066],[298,0.0057],[299,0.0061],[300,0.0073],[301,0.0073],[303,0.0079],[304,0.009],[305,0.0096],[307,0.0097],[310,0.0094],[311,0.0083],[312,0.0084],[313,0.0093],[316,0.0097],[318,0.0092],[319,0.0094],[320,0.0093],[321,0.0094],[322,0.0093],[323,0.009],[324,0.0099],[325,0.01],[326,0.0098],[327,0.0095],[329,0.0097],[330,0.0093],[331,0.0094],[332,0.01],[333,0.0096],[334,0.01],[335,0.0099],[337,0.0101],[339,0.0099],[340,0.0094],[341,0.0088],[342,0.0088],[343,0.0091],[345,0.0089],[346,0.009],[348,0.0087],[349,0.0088],[350,0.0084],[351,0.009],[353,0.0089],[354,0.0085],[355,0.0083],[358,0.0085],[360,0.0083]]
       }
       CTRL_R_mouth_lipBiteD {
        type=MovieSceneScriptingFloatChannel
        default=0
        thinned_from=492
        keys [[-131,0.0022],[-130,0.0016],[-129,0.0008],[-127,0.0001],[-126,0.0001],[-125,0.0003],[-124,0.0008],[-123,0.0005],[-122,0.0004],[-121,0],[-120,0.0008],[-119,0.0005],[-117,0.0007],[-116,0.001],[-114,0.0011],[-113,0.0009],[-111,0.0011],[-109,0.0004],[-107,0.0015],[-101,0.0011],[-100,0.0007],[-99,0.0008],[-97,0.0013],[-95,0.0012],[-93,0.0013],[-92,0.001],[-91,0.0009],[-89,0.0016],[-88,0.0017],[-85,0.0027],[-84,0.0029],[-83,0.0037],[-82,0.0048],[-81,0.008],[-80,0.028],[-79,0.033],[-78,0.0785],[-77,0.1485],[-76,0.1975],[-75,0.2681],[-74,0.3284],[-72,0.3752],[-71,0.384],[-70,0.3916],[-69,0.4044],[-67,0.4024],[-66,0.3976],[-65,0.3888],[-64,0.3849],[-62,0.3883],[-61,0.3786],[-60,0.3591],[-59,0.3183],[-58,0.243],[-57,0.2296],[-56,0.1956],[-55,0.1689],[-54,0.1522],[-53,0.1409],[-52,0.1358],[-51,0.1065],[-50,0.0898],[-49,0.0883],[-48,0.0639],[-46,0.0646],[-45,0.0603],[-44,0.0586],[-43,0.0531],[-42,0.0538],[-41,0.0605],[-40,0.0733],[-39,0.1005],[-38,0.111],[-37,0.2451],[-36,0.309],[-35,0.3391],[-34,0.3409],[-33,0.3316],[-32,0.283],[-31,0.2599],[-30,0.1408],[-29,0.1332],[-28,0.1176],[-27,0.1119],[-26,0.1105],[-25,0.1119],[-24,0.1127],[-23,0.1168],[-22,0.1029],[-21,0.0857],[-20,0.0834],[-19,0.0813],[-18,0.0721],[-17,0.0637],[-16,0.0537],[-15,0.0456],[-14,0.0416],[-13,0.0413],[-12,0.0392],[-11,0.0376],[-10,0.0385],[-9,0.0365],[-8,0.0329],[-7,0.0336],[-6,0.0356],[-5,0.0365],[-4,0.0355],[-3,0.0342],[-2,0.0353],[-1,0.0327],[0,0.0297],[1,0.0302],[3,0.0306],[4,0.0301],[5,0.0279],[6,0.0237],[8,0.0231],[10,0.0205],[11,0.0207],[13,0.0242],[15,0.0231],[16,0.0238],[17,0.0242],[19,0.0245],[20,0.0254],[21,0.0258],[24,0.0257],[25,0.027],[26,0.0287],[27,0.0301],[28,0.0304],[29,0.03],[30,0.0293],[31,0.0289],[32,0.0292],[33,0.0301],[34,0.0287],[36,0.0288],[37,0.0312],[38,0.0305],[39,0.0296],[40,0.0302],[41,0.0292],[42,0.0285],[43,0.031],[44,0.0331],[45,0.0333],[46,0.03],[47,0.0316],[48,0.0387],[50,0.0437],[51,0.0439],[52,0.0451],[53,0.049],[54,0.0462],[55,0.0385],[56,0.0338],[57,0.0318],[59,0.023],[60,0.0223],[62,0.0227],[63,0.0224],[64,0.0176],[65,0.0176],[66,0.0186],[67,0.0209],[68,0.0208],[73,0.0217],[74,0.0226],[75,0.025],[76,0.0262],[77,0.0256],[78,0.0258],[79,0.0277],[80,0.0283],[81,0.0277],[82,0.0283],[83,0.0304],[84,0.0316],[85,0.0308],[86,0.0285],[87,0.0281],[88,0.0284],[89,0.0282],[90,0.0286],[91,0.0287],[92,0.0306],[93,0.0306],[94,0.0295],[95,0.026],[96,0.0255],[97,0.0258],[98,0.025],[99,0.0248],[100,0.0238],[101,0.0231],[103,0.023],[104,0.0226],[105,0.0226],[106,0.0211],[107,0.0214],[108,0.0214],[109,0.0204],[110,0.0198],[112,0.0169],[113,0.0152],[114,0.0138],[115,0.0135],[116,0.0151],[117,0.0147],[118,0.0153],[119,0.0172],[120,0.0186],[121,0.0183],[122,0.0166],[123,0.0166],[124,0.0164],[125,0.0166],[126,0.0181],[127,0.0184],[128,0.0174],[129,0.0169],[130,0.0178],[131,0.0184],[132,0.0171],[133,0.0165],[134,0.0162],[135,0.0156],[136,0.0146],[137,0.0143],[138,0.0142],[139,0.0135],[140,0.0172],[141,0.019],[142,0.0195],[144,0.0165],[145,0.0166],[147,0.0155],[148,0.0166],[149,0.0158],[150,0.0142],[152,0.0164],[153,0.018],[154,0.0184],[155,0.0175],[156,0.017],[157,0.0159],[158,0.0157],[159,0.0174],[160,0.0179],[162,0.0161],[163,0.0149],[164,0.0141],[166,0.0151],[167,0.0144],[168,0.0126],[170,0.0098],[171,0.0087],[172,0.009],[173,0.0091],[174,0.0082],[175,0.0069],[176,0.0076],[178,0.0079],[179,0.0062],[180,0.005],[182,0.0046],[184,0.0033],[187,0.0017],[189,0.0014],[190,0.0009],[191,0.001],[192,0.0009],[193,0.0017],[194,0.0039],[195,0.004],[196,0.0032],[197,0.0021],[198,0.0023],[199,0.0036],[200,0.004],[201,0.0052],[202,0.0103],[203,0.0111],[204,0.007],[205,0.0069],[208,0.0056],[209,0.0057],[211,0.0046],[213,0.0043],[215,0.0043],[216,0.0047],[218,0.0038],[219,0.004],[220,0.0049],[222,0.0059],[223,0.0068],[224,0.0073],[225,0.0065],[226,0.0065],[227,0.0068],[228,0.0062],[229,0.0062],[231,0.0072],[232,0.0073],[233,0.0083],[234,0.0079],[235,0.0043],[236,0.0027],[238,0.0024],[239,0.0015],[240,0.0012],[241,0.0016],[242,0.0017],[244,0.0009],[245,0.0011],[249,0.0004],[250,0.0004],[252,0.0008],[253,0.0004],[255,0.0011],[256,0.0007],[257,0],[259,0],[261,0.0021],[266,0.0028],[271,0.0023],[272,0.0021],[274,0.0009],[276,0.0006],[278,0.0007],[279,0.0005],[280,0.0005],[281,0.0008],[284,0.0006],[285,0.0007],[286,0],[288,0.0006],[289,0.0007],[291,0.0003],[294,0.0005],[296,0.0012],[298,0.0014],[300,0.002],[302,0.0018],[303,0.0023],[305,0.0059],[306,0.0071],[307,0.0076],[310,0.0077],[311,0.0083],[312,0.0084],[313,0.0089],[314,0.0088],[315,0.0093],[316,0.0092],[318,0.0085],[319,0.0084],[320,0.0085],[321,0.0082],[322,0.0075],[323,0.0071],[325,0.0084],[327,0.0093],[329,0.0094],[331,0.0088],[332,0.0083],[334,0.0081],[336,0.0087],[337,0.0084],[338,0.0079],[340,0.0082],[341,0.0075],[342,0.0073],[343,0.0076],[344,0.0075],[345,0.0077],[346,0.0077],[347,0.0074],[348,0.0076],[349,0.0074],[351,0.0065],[352,0.0068],[353,0.0068],[355,0.0058],[356,0.0057],[358,0.0063],[359,0.0057],[360,0.0057]]
       }
       CTRL_R_mouth_funnelD {
        type=MovieSceneScriptingFloatChannel
        default=0
        thinned_from=492
        keys [[-131,0.0145],[-130,0.0182],[-129,0.017],[-128,0.0174],[-127,0.0194],[-126,0.02],[-125,0.0212],[-124,0.0238],[-123,0.0269],[-122,0.0276],[-121,0.0259],[-120,0.0253],[-119,0.024],[-118,0.0233],[-117,0.0228],[-116,0.02],[-115,0.0207],[-113,0.0189],[-112,0.0193],[-111,0.0188],[-110,0.0165],[-109,0.0149],[-108,0.015],[-107,0.0138],[-106,0.0136],[-105,0.0131],[-104,0.0143],[-103,0.0123],[-102,0.0126],[-101,0.0103],[-100,0.0086],[-99,0.009],[-97,0.0081],[-96,0.0075],[-95,0.0074],[-94,0.01],[-93,0.0101],[-92,0.0111],[-91,0.0102],[-90,0.0094],[-89,0.0092],[-88,0.0087],[-87,0.007],[-86,0.0051],[-85,0.0043],[-83,0.0036],[-82,0.0035],[-81,0.0036],[-80,0.0083],[-79,0.0092],[-77,0.0267],[-76,0.0338],[-75,0.0291],[-74,0.0202],[-73,0.0159],[-72,0.0132],[-70,0.0112],[-68,0.0101],[-67,0.0081],[-66,0.01],[-65,0.0096],[-64,0.0088],[-63,0.007],[-62,0.0056],[-60,0.004],[-58,0.0035],[-56,0.0032],[-55,0.0028],[-49,0.0027],[-43,0.0032],[-41,0.0031],[-38,0.0035],[-37,0.0159],[-36,0.0138],[-35,0.0095],[-34,0.006],[-33,0.0043],[-32,0.0039],[-31,0.0037],[-30,0.0031],[-28,0.0027],[-27,0.0028],[-26,0.0024],[-23,0.0021],[-22,0.0026],[-18,0.0031],[-17,0.003],[-16,0.0033],[-14,0.0032],[-13,0.0033],[-12,0.0031],[-11,0.0033],[-9,0.003],[-7,0.0033],[-4,0.003],[-3,0.0031],[-1,0.003],[0,0.0031],[1,0.0029],[4,0.0029],[5,0.0028],[6,0.0031],[19,0.0029],[20,0.0031],[32,0.0029],[33,0.0028],[39,0.0028],[40,0.003],[42,0.0029],[51,0.003],[52,0.0031],[53,0.0027],[55,0.0028],[56,0.0027],[58,0.0008],[59,0.0012],[62,0.0019],[65,0.0017],[66,0.0022],[68,0.0016],[69,0.002],[71,0.0018],[72,0.0019],[73,0.0016],[74,0.0018],[75,0.0023],[76,0.0019],[78,0.0019],[80,0.0014],[81,0.0014],[82,0.001],[83,0.0011],[85,0.001],[86,0.0011],[87,0.0015],[88,0.001],[89,0.0011],[90,0.0009],[94,0.0007],[98,0.0008],[100,0.0017],[102,0.0014],[103,0.0009],[105,0.0009],[106,0.0007],[107,0.0008],[108,0.0008],[111,0.0013],[113,0.0011],[114,0.0005],[115,0.0011],[116,0.0012],[117,0.0007],[118,0.0011],[119,0.0009],[122,0.0012],[123,0.0009],[124,0.0011],[125,0.0005],[126,0.0003],[127,0.0006],[128,0.0004],[129,0.0006],[132,0.0001],[133,0.0007],[134,0.0011],[136,0.0012],[137,0.0021],[138,0.0022],[139,0.0009],[140,0.0017],[141,0.0014],[143,0.0018],[145,0.0016],[146,0.0012],[147,0.0015],[148,0.0011],[149,0.001],[150,0.0012],[151,0.0015],[152,0.0013],[154,0.0016],[155,0.002],[156,0.0014],[158,0.0015],[159,0.0018],[160,0.0013],[162,0.0018],[163,0.0018],[164,0.0014],[166,0.0018],[168,0.0016],[169,0.0025],[170,0.0028],[171,0.0033],[172,0.0029],[173,0.0021],[174,0.0017],[177,0.0022],[178,0.0022],[179,0.0014],[181,0.0016],[184,0.0009],[186,0.0014],[188,0.0014],[189,0.0011],[190,0.0011],[191,0.0013],[192,0.0008],[193,0.002],[194,0.003],[195,0.003],[196,0.0025],[197,0.0017],[199,0.0027],[200,0.0027],[201,0.0024],[202,0.0032],[203,0.0029],[207,0.0025],[209,0.0029],[210,0.0028],[212,0.0028],[214,0.0023],[216,0.0026],[217,0.0022],[219,0.0024],[220,0.0027],[224,0.0023],[230,0.0027],[232,0.0025],[233,0.0026],[234,0.0025],[235,0.0013],[236,0.0028],[237,0.0031],[238,0.0031],[239,0.002],[240,0.0019],[242,0.0021],[243,0.0014],[244,0.0013],[245,0.0017],[246,0.0012],[247,0.0052],[248,0.0237],[249,0.0365],[250,0.028],[251,0.0225],[252,0.0067],[253,0.0029],[254,0.0019],[257,0.0004],[258,0.0025],[259,0.0032],[260,0.0079],[261,0.0051],[262,0.0039],[263,0.0037],[266,0.0038],[267,0.0057],[268,0.0084],[269,0.0085],[270,0.0084],[271,0.0093],[272,0.0055],[273,0.0052],[275,0.0064],[276,0.0062],[277,0.0068],[279,0.0085],[280,0.0089],[281,0.0085],[282,0.0091],[283,0.0086],[284,0.0084],[285,0.0081],[286,0.0058],[287,0.0057],[289,0.0051],[290,0.0052],[291,0.0056],[292,0.0077],[294,0.0138],[295,0.0115],[296,0.0075],[297,0.0068],[298,0.0026],[299,0.0021],[300,0.0024],[304,0.0023],[307,0.0028],[309,0.0024],[316,0.0022],[317,0.0019],[318,0.0018],[320,0.0022],[321,0.0022],[322,0.0016],[323,0.0015],[325,0.0023],[326,0.0024],[327,0.0023],[332,0.0024],[334,0.0022],[338,0.0024],[340,0.0022],[341,0.0019],[343,0.0023],[344,0.0021],[348,0.0022],[350,0.0017],[351,0.0022],[352,0.0021],[353,0.0024],[354,0.0022],[356,0.0025],[357,0.0023],[358,0.0024],[359,0.0021],[360,0.0021]]
       }
       CTRL_L_mouth_pressU {type=MovieSceneScriptingFloatChannel, default=0, const [0,492]}
       CTRL_L_mouth_funnelD {
        type=MovieSceneScriptingFloatChannel
        default=0
        thinned_from=492
        keys [[-131,0.0122],[-130,0.0157],[-129,0.0146],[-128,0.0152],[-127,0.017],[-126,0.0175],[-125,0.0186],[-124,0.021],[-123,0.0237],[-122,0.0244],[-121,0.0227],[-120,0.0221],[-119,0.021],[-118,0.0202],[-117,0.0197],[-116,0.0168],[-115,0.0174],[-113,0.0153],[-112,0.0153],[-111,0.0148],[-110,0.0128],[-109,0.0115],[-108,0.0116],[-107,0.0106],[-105,0.0102],[-104,0.0112],[-103,0.0095],[-102,0.01],[-101,0.0078],[-100,0.0063],[-99,0.0065],[-98,0.0062],[-97,0.0056],[-96,0.0052],[-95,0.0052],[-94,0.0076],[-93,0.0077],[-92,0.0084],[-91,0.0075],[-90,0.0069],[-89,0.0066],[-88,0.0058],[-87,0.0043],[-86,0.0026],[-85,0.0018],[-82,0.0013],[-81,0.0014],[-80,0.0057],[-79,0.0066],[-78,0.0146],[-77,0.0223],[-76,0.0282],[-75,0.0237],[-74,0.0159],[-73,0.0123],[-72,0.0103],[-71,0.0094],[-68,0.0077],[-67,0.006],[-66,0.0076],[-64,0.0065],[-63,0.0048],[-62,0.0035],[-61,0.0029],[-60,0.002],[-56,0.002],[-55,0.0017],[-54,0.0016],[-51,0.002],[-46,0.002],[-45,0.0023],[-41,0.0016],[-38,0.0019],[-37,0.0117],[-36,0.0099],[-35,0.0063],[-34,0.0033],[-33,0.0022],[-31,0.0021],[-30,0.0018],[-27,0.0018],[-23,0.0011],[-19,0.0017],[-18,0.0022],[-17,0.0019],[-15,0.0018],[-14,0.0021],[-13,0.0018],[-11,0.0022],[-9,0.0018],[-7,0.0021],[-4,0.002],[-3,0.0021],[1,0.002],[2,0.0021],[6,0.0018],[15,0.0017],[17,0.0019],[18,0.0018],[19,0.0019],[20,0.0018],[21,0.0019],[23,0.0018],[24,0.002],[33,0.0017],[35,0.0018],[36,0.0016],[40,0.0019],[45,0.0021],[46,0.0018],[50,0.0017],[51,0.0014],[52,0.0017],[53,0.0013],[56,0.0014],[58,0.0002],[63,0.0006],[65,0.0004],[66,0.0008],[68,0.0004],[69,0.0009],[72,0.0009],[73,0.0006],[74,0.0006],[75,0.0011],[76,0.0007],[78,0.0009],[80,0.0004],[81,0.0006],[82,0.0002],[83,0.0004],[84,0.0003],[85,0.0006],[86,0.0005],[87,0.0007],[88,0.0002],[90,0.0002],[91,0],[96,0.0001],[98,0],[100,0.0007],[102,0.0005],[103,0],[106,0],[107,0.0002],[108,0],[109,0.0002],[110,0],[111,0.0003],[113,0],[116,0.0002],[117,0],[133,0],[134,0.0003],[135,0.0001],[136,0.0002],[137,0.0009],[138,0.0009],[139,0.0003],[140,0.001],[141,0.0007],[142,0.0006],[143,0.0009],[145,0.0002],[146,0],[147,0.0005],[148,0.0001],[150,0],[151,0.0005],[153,0.0003],[155,0.0006],[156,0.0002],[158,0.0004],[159,0.0008],[160,0.0003],[162,0.0008],[163,0.0008],[164,0.0004],[165,0.0006],[166,0.0005],[167,0.0008],[168,0.0005],[169,0.0013],[171,0.0017],[172,0.0013],[173,0.0005],[175,0],[176,0.0003],[178,0.0006],[179,0.0002],[180,0],[181,0.0001],[191,0],[192,0.0001],[194,0.0015],[195,0.0014],[197,0.0003],[198,0.001],[199,0.0013],[200,0.0013],[201,0.001],[202,0.0018],[204,0.0013],[210,0.0012],[213,0.0014],[214,0.0011],[216,0.0014],[217,0.0011],[219,0.001],[220,0.0015],[222,0.0012],[223,0.0014],[225,0.0012],[227,0.0014],[232,0.001],[233,0.0011],[234,0.001],[235,0.0003],[236,0.0013],[237,0.0015],[238,0.0014],[239,0.0005],[242,0.0007],[243,0.0002],[244,0],[245,0.0003],[246,0.0002],[247,0.0037],[248,0.0203],[249,0.0323],[250,0.0248],[251,0.0194],[252,0.0052],[253,0.0019],[254,0.0006],[256,0],[257,0],[258,0.0015],[259,0.0022],[260,0.0069],[261,0.0039],[262,0.0027],[266,0.0029],[267,0.0046],[268,0.0067],[270,0.0065],[271,0.0073],[272,0.0036],[273,0.0034],[274,0.0036],[275,0.0043],[276,0.0041],[277,0.0047],[279,0.0064],[280,0.0068],[281,0.0063],[282,0.0067],[283,0.0061],[284,0.006],[285,0.0057],[286,0.0036],[289,0.0031],[290,0.0031],[291,0.0037],[292,0.0055],[293,0.0082],[294,0.0111],[295,0.009],[296,0.0055],[297,0.0048],[298,0.0014],[299,0.0011],[300,0.0014],[303,0.0012],[307,0.0018],[309,0.0015],[316,0.0013],[317,0.0009],[321,0.0014],[322,0.001],[323,0.0008],[324,0.0014],[327,0.0011],[330,0.0013],[331,0.0015],[333,0.0013],[335,0.0014],[337,0.0012],[338,0.0014],[339,0.0011],[340,0.0012],[342,0.0008],[343,0.0011],[344,0.0008],[346,0.0012],[348,0.0012],[350,0.0009],[351,0.0013],[352,0.0011],[360,0.0012]]
       }
       CTRL_L_mouth_suckBlow {
        type=MovieSceneScriptingFloatChannel
        default=0
        thinned_from=492
        keys [[-131,0.0049],[-130,0.0048],[-129,0.0042],[-128,0.0039],[-127,0.0041],[-126,0.0046],[-125,0.0049],[-102,0.0049],[-100,0.0039],[-99,0.0046],[-98,0.0049],[-73,0.0049],[-71,0.0039],[-69,0.0049],[-67,0.0039],[-63,0.004],[-62,0.0044],[-60,0.0045],[-59,0.0054],[-58,0.0098],[-57,0.0112],[-56,0.0144],[-55,0.016],[-54,0.0169],[-53,0.0171],[-52,0.017],[-51,0.015],[-50,0.0125],[-49,0.0122],[-47,0.0083],[-46,0.0077],[-45,0.0049],[-32,0.0049],[-31,0.0058],[-30,0.0117],[-29,0.0119],[-28,0.0124],[-27,0.0127],[-26,0.0127],[-25,0.0134],[-24,0.0137],[-23,0.0127],[-22,0.0077],[-21,0.005],[35,0.0049],[36,0.0054],[38,0.0049],[40,0.0049],[41,0.0052],[42,0.0063],[43,0.0063],[44,0.0055],[45,0.0054],[46,0.0058],[48,0.0049],[137,0.0049],[138,0.0088],[139,0.0088],[140,0.0051],[240,0.0049],[241,0.0045],[245,0.0044],[246,0.0049],[360,0.0049]]
       }
       CTRL_L_mouth_pressD {type=MovieSceneScriptingFloatChannel, default=0, const [0,492]}
       CTRL_L_mouth_funnelU {
        type=MovieSceneScriptingFloatChannel
        default=0
        thinned_from=492
        keys [[-131,0.0024],[-130,0.0012],[-129,0.001],[-126,0],[-124,0.0005],[-121,0],[-120,0.0005],[-119,0.0002],[-118,0.0007],[-115,0.001],[-113,0.0008],[-111,0.0009],[-109,0.0003],[-107,0.0014],[-106,0.0015],[-105,0.0012],[-104,0.0011],[-103,0.0008],[-101,0.0009],[-100,0.0005],[-99,0.0009],[-95,0.0012],[-94,0.0008],[-91,0.0005],[-89,0.0012],[-88,0.0015],[-87,0.0021],[-80,0.0029],[-77,0.0026],[-76,0.0029],[-75,0.0026],[-74,0.0025],[-72,0.0027],[-71,0.0025],[-70,0.0028],[-69,0.0026],[-67,0.0028],[-65,0.0025],[-61,0.0026],[-58,0.0029],[-54,0.0027],[-51,0.0028],[-49,0.003],[-43,0.0028],[-42,0.0026],[-38,0.0028],[-37,0.0026],[-35,0.0027],[-34,0.0025],[-32,0.0028],[-28,0.0024],[-26,0.0024],[-24,0.0021],[-20,0.0028],[-19,0.0026],[-18,0.003],[-15,0.0025],[-8,0.0028],[-2,0.0025],[2,0.0027],[3,0.0025],[5,0.0029],[9,0.0029],[12,0.0028],[16,0.0029],[20,0.0029],[28,0.0027],[29,0.0029],[33,0.0027],[43,0.003],[45,0.0028],[46,0.0029],[47,0.0028],[51,0.0026],[52,0.0028],[53,0.0026],[54,0.0027],[56,0.0024],[57,0.0017],[58,0.0007],[59,0.0011],[62,0.0017],[65,0.0016],[66,0.0022],[68,0.0014],[69,0.0019],[72,0.0017],[73,0.0014],[74,0.0016],[75,0.002],[76,0.0017],[79,0.0015],[80,0.0013],[81,0.0016],[82,0.0014],[83,0.0014],[85,0.001],[86,0.0013],[87,0.0014],[88,0.0008],[89,0.001],[91,0.0008],[92,0.0011],[94,0.0006],[98,0.0005],[100,0.0014],[102,0.0013],[103,0.0007],[105,0.0009],[106,0.0007],[107,0.001],[108,0.0005],[111,0.0011],[112,0.0009],[113,0.0009],[114,0.0005],[115,0.0011],[116,0.0012],[117,0.0006],[122,0.001],[123,0.0008],[124,0.0009],[125,0.0004],[127,0.0009],[128,0.0004],[130,0.0007],[132,0.0002],[133,0.0005],[134,0.0013],[136,0.0013],[137,0.0021],[138,0.0088],[139,0.0076],[140,0.0096],[141,0.0097],[142,0.0111],[143,0.011],[144,0.0098],[145,0.0069],[146,0.0079],[147,0.0076],[148,0.009],[149,0.0054],[150,0.0069],[151,0.0089],[152,0.0085],[153,0.0104],[154,0.0108],[155,0.008],[156,0.0078],[157,0.008],[158,0.0069],[159,0.0074],[160,0.0072],[161,0.0081],[162,0.0078],[163,0.0072],[164,0.0055],[165,0.0059],[166,0.0043],[167,0.0029],[168,0.0021],[169,0.0026],[171,0.0027],[172,0.0024],[173,0.0015],[174,0.0009],[178,0.0017],[179,0.0009],[180,0.0008],[182,0.0009],[183,0.0005],[184,0.0006],[185,0.0009],[186,0.0009],[188,0.0011],[190,0.0006],[192,0.0006],[193,0.0018],[194,0.0028],[195,0.0028],[197,0.0018],[198,0.0025],[199,0.0028],[200,0.0028],[201,0.0023],[202,0.003],[204,0.0023],[207,0.0024],[209,0.0022],[211,0.0024],[214,0.0023],[216,0.0024],[217,0.0021],[219,0.0022],[220,0.0027],[222,0.0021],[223,0.0023],[226,0.0023],[227,0.0025],[231,0.002],[232,0.0021],[233,0.0025],[234,0.0023],[235,0.001],[236,0.0022],[237,0.0026],[238,0.0025],[239,0.0015],[240,0.0014],[242,0.0016],[244,0.0009],[245,0.0012],[246,0.0009],[247,0.0008],[248,0.0003],[249,0.0006],[250,0.0002],[251,0.0008],[252,0.0004],[253,0.0005],[254,0.001],[255,0.0011],[256,0.0007],[257,0],[259,0.0002],[260,0.0014],[261,0.002],[262,0.0019],[263,0.0028],[265,0.0031],[268,0.0029],[269,0.0027],[271,0.0025],[273,0.0015],[274,0.0007],[279,0.0003],[281,0.0006],[282,0.0004],[283,0.0007],[285,0.0007],[286,0.0001],[288,0.0006],[291,0.0002],[293,0.0007],[294,0.0007],[296,0.0013],[298,0.0015],[300,0.0023],[304,0.0023],[307,0.0027],[309,0.0024],[310,0.0024],[311,0.002],[314,0.002],[316,0.0021],[318,0.0018],[320,0.0021],[321,0.002],[323,0.0015],[324,0.002],[326,0.0022],[327,0.0019],[328,0.0021],[331,0.002],[332,0.0021],[333,0.0019],[335,0.0021],[336,0.002],[338,0.0022],[341,0.0018],[343,0.0021],[346,0.0018],[347,0.0019],[350,0.0016],[351,0.002],[352,0.002],[353,0.0021],[357,0.0018],[358,0.0019],[360,0.0018]]
       }
       CTRL_R_mouth_towardsU {
        type=MovieSceneScriptingFloatChannel
        default=0
        thinned_from=492
        keys [[-131,0.0144],[-130,0.0144],[-129,0.0142],[-128,0.0136],[-127,0.014],[-126,0.0137],[-124,0.0153],[-123,0.0147],[-122,0.0145],[-121,0.0141],[-120,0.0146],[-117,0.0144],[-116,0.0141],[-115,0.0144],[-114,0.0151],[-113,0.0153],[-111,0.0152],[-110,0.0146],[-109,0.0146],[-108,0.0143],[-106,0.0141],[-105,0.0131],[-104,0.014],[-103,0.014],[-101,0.0147],[-100,0.0136],[-99,0.0143],[-98,0.0147],[-97,0.0148],[-95,0.0147],[-94,0.0153],[-93,0.0152],[-92,0.0146],[-91,0.0147],[-90,0.0154],[-89,0.0164],[-88,0.0171],[-87,0.017],[-86,0.016],[-84,0.0126],[-80,0.0078],[-79,0.0077],[-78,0.0071],[-77,0.0061],[-76,0.006],[-75,0.0062],[-74,0.0069],[-73,0.0085],[-72,0.0089],[-70,0.0088],[-69,0.0096],[-68,0.0094],[-67,0.01],[-65,0.0097],[-63,0.0106],[-61,0.012],[-60,0.0132],[-59,0.0153],[-58,0.0193],[-57,0.0198],[-56,0.021],[-55,0.0208],[-53,0.0197],[-52,0.0194],[-51,0.0197],[-50,0.0194],[-49,0.0194],[-48,0.0185],[-47,0.0166],[-46,0.0163],[-45,0.0137],[-44,0.013],[-43,0.0106],[-42,0.0096],[-41,0.0089],[-39,0.0085],[-37,0.0088],[-36,0.0095],[-35,0.0104],[-34,0.0103],[-33,0.0113],[-32,0.0141],[-31,0.0148],[-30,0.0172],[-24,0.0176],[-23,0.0172],[-22,0.0197],[-21,0.0193],[-20,0.0179],[-19,0.0145],[-18,0.0127],[-17,0.0133],[-16,0.0132],[-15,0.0135],[-13,0.0132],[-11,0.0136],[-10,0.0128],[-9,0.0136],[-8,0.0136],[-7,0.0132],[-6,0.0133],[-4,0.0127],[-3,0.0125],[-2,0.0131],[-1,0.0135],[0,0.013],[1,0.0136],[2,0.0137],[3,0.0134],[5,0.0132],[6,0.0125],[7,0.0128],[9,0.0128],[10,0.0126],[13,0.013],[15,0.0127],[16,0.0124],[17,0.0125],[19,0.0139],[20,0.0141],[21,0.014],[22,0.0133],[28,0.0131],[30,0.0134],[32,0.0132],[33,0.0126],[35,0.0125],[37,0.0126],[42,0.0137],[45,0.0136],[46,0.0145],[47,0.0141],[48,0.0155],[49,0.0154],[50,0.0144],[51,0.0154],[52,0.0167],[53,0.0178],[54,0.0181],[55,0.018],[56,0.0174],[57,0.0163],[58,0.0146],[59,0.0142],[60,0.0143],[61,0.0152],[63,0.0154],[64,0.016],[65,0.016],[66,0.0171],[67,0.0166],[68,0.0156],[69,0.0164],[71,0.016],[72,0.0162],[73,0.0153],[75,0.0169],[76,0.0164],[78,0.0163],[80,0.0152],[81,0.0162],[82,0.016],[84,0.0149],[86,0.0157],[87,0.0153],[88,0.0138],[90,0.0141],[91,0.0139],[92,0.0146],[94,0.0145],[95,0.0142],[96,0.0141],[98,0.0133],[100,0.0144],[101,0.0138],[102,0.0135],[105,0.0136],[106,0.0141],[107,0.0143],[108,0.0137],[109,0.0138],[110,0.0145],[111,0.0145],[112,0.0147],[113,0.0137],[114,0.0134],[115,0.0145],[116,0.0141],[118,0.0157],[119,0.0146],[120,0.0147],[121,0.0145],[122,0.0149],[123,0.0148],[124,0.015],[125,0.0144],[127,0.0154],[128,0.0152],[129,0.0155],[130,0.0166],[131,0.0156],[132,0.0153],[134,0.0178],[136,0.0199],[137,0.0213],[138,0.0198],[139,0.0154],[140,0.0166],[142,0.0168],[143,0.0174],[144,0.0172],[145,0.016],[146,0.0153],[147,0.0164],[148,0.0157],[149,0.0147],[151,0.016],[153,0.0159],[155,0.0163],[156,0.0154],[157,0.0155],[158,0.0148],[159,0.0162],[160,0.0151],[161,0.0149],[162,0.015],[163,0.0139],[165,0.0139],[166,0.0132],[169,0.013],[171,0.0122],[172,0.0123],[174,0.0115],[175,0.0118],[176,0.0111],[178,0.0115],[179,0.0101],[180,0.01],[181,0.0112],[182,0.0113],[183,0.0108],[184,0.0101],[186,0.0111],[187,0.0124],[189,0.0128],[190,0.0123],[191,0.0124],[192,0.0113],[193,0.0126],[194,0.0146],[195,0.0132],[196,0.009],[197,0.0063],[198,0.0075],[199,0.0081],[200,0.0079],[201,0.0074],[202,0.0083],[203,0.0082],[205,0.0099],[206,0.011],[208,0.0123],[212,0.0127],[214,0.0137],[216,0.0137],[217,0.0142],[218,0.0143],[219,0.0142],[220,0.0146],[221,0.014],[222,0.0142],[223,0.0139],[224,0.0143],[225,0.0141],[227,0.0143],[228,0.0137],[230,0.0116],[232,0.0099],[233,0.0084],[234,0.0082],[235,0.0071],[236,0.0081],[237,0.0099],[238,0.011],[239,0.0094],[240,0.0101],[241,0.0123],[242,0.0127],[243,0.0129],[244,0.0136],[246,0.0114],[247,0.0132],[248,0.0156],[249,0.016],[250,0.0147],[251,0.015],[252,0.0127],[253,0.0113],[254,0.0125],[257,0.0125],[258,0.015],[259,0.0159],[260,0.0237],[261,0.0247],[263,0.0233],[264,0.0232],[265,0.0228],[266,0.0227],[267,0.0206],[268,0.019],[269,0.0139],[270,0.0129],[271,0.0101],[272,0.0114],[273,0.0119],[274,0.0118],[275,0.0128],[279,0.0136],[280,0.0143],[282,0.014],[283,0.0144],[285,0.0145],[286,0.0126],[288,0.013],[290,0.0125],[291,0.0121],[292,0.012],[293,0.013],[294,0.0126],[295,0.0114],[296,0.0108],[297,0.0107],[298,0.0091],[299,0.0095],[300,0.0113],[301,0.0115],[302,0.0128],[303,0.0133],[304,0.0148],[305,0.0156],[306,0.0162],[308,0.0164],[310,0.016],[311,0.0154],[312,0.0155],[313,0.0159],[314,0.0165],[315,0.0167],[317,0.0164],[318,0.0156],[319,0.0159],[320,0.0161],[322,0.0157],[323,0.0153],[324,0.016],[326,0.0169],[328,0.017],[330,0.0163],[331,0.0165],[332,0.0171],[333,0.0165],[334,0.0167],[335,0.0166],[336,0.017],[337,0.017],[339,0.0169],[340,0.0165],[341,0.0158],[344,0.0161],[346,0.0159],[347,0.0156],[348,0.0156],[349,0.016],[350,0.0157],[351,0.0158],[352,0.0163],[353,0.0165],[354,0.0162],[356,0.0164],[358,0.017],[359,0.0164],[360,0.0164]]
       }
       CTRL_R_mouth_purseD {
        type=MovieSceneScriptingFloatChannel
        default=0
        thinned_from=492
        keys [[-131,0.0662],[-130,0.0658],[-129,0.057],[-128,0.0521],[-127,0.0515],[-126,0.0512],[-125,0.0529],[-124,0.0534],[-123,0.0541],[-122,0.0544],[-121,0.055],[-120,0.0573],[-119,0.0581],[-118,0.0566],[-117,0.0558],[-116,0.0504],[-115,0.0485],[-113,0.046],[-112,0.0438],[-111,0.0414],[-109,0.0411],[-108,0.0387],[-107,0.0352],[-106,0.0346],[-105,0.0335],[-104,0.0357],[-103,0.0359],[-102,0.0358],[-101,0.037],[-100,0.0349],[-99,0.0381],[-98,0.0377],[-97,0.0364],[-95,0.0374],[-94,0.0461],[-93,0.0465],[-92,0.0459],[-91,0.0505],[-90,0.0568],[-89,0.0624],[-88,0.0724],[-87,0.0899],[-86,0.1037],[-85,0.1188],[-84,0.1358],[-83,0.1572],[-82,0.209],[-81,0.2589],[-80,0.2979],[-79,0.3033],[-78,0.3498],[-77,0.3553],[-76,0.3675],[-75,0.3761],[-74,0.3723],[-73,0.3791],[-72,0.3828],[-71,0.3765],[-69,0.3718],[-68,0.3643],[-67,0.3583],[-66,0.3651],[-65,0.3595],[-64,0.3523],[-63,0.3401],[-62,0.3271],[-61,0.3065],[-60,0.259],[-59,0.1836],[-58,0.0996],[-57,0.0831],[-56,0.0489],[-55,0.0343],[-54,0.0258],[-53,0.0208],[-52,0.0184],[-50,0.017],[-49,0.0168],[-48,0.0205],[-46,0.0257],[-45,0.0436],[-44,0.0519],[-43,0.0832],[-42,0.1229],[-41,0.16],[-40,0.2275],[-39,0.2942],[-38,0.3097],[-37,0.3365],[-36,0.3293],[-35,0.3281],[-34,0.3118],[-33,0.2685],[-32,0.179],[-31,0.157],[-30,0.06],[-29,0.0522],[-28,0.0353],[-27,0.0252],[-26,0.0206],[-25,0.0169],[-24,0.0157],[-23,0.0177],[-22,0.0279],[-21,0.0324],[-20,0.0261],[-19,0.0264],[-18,0.0328],[-17,0.0269],[-16,0.0233],[-15,0.025],[-14,0.0307],[-13,0.03],[-12,0.0269],[-11,0.0249],[-10,0.0216],[-9,0.0217],[-8,0.0211],[-7,0.0179],[-6,0.0167],[-4,0.0162],[-3,0.0163],[-2,0.0146],[-1,0.0152],[1,0.0146],[3,0.0146],[4,0.0148],[5,0.0155],[6,0.0159],[7,0.0156],[8,0.0158],[9,0.0164],[10,0.0164],[12,0.0156],[13,0.0162],[14,0.0165],[15,0.0171],[17,0.0177],[18,0.0165],[19,0.0176],[20,0.0205],[21,0.019],[22,0.0184],[27,0.0171],[28,0.0164],[29,0.0172],[30,0.0165],[31,0.0155],[32,0.0168],[33,0.0168],[35,0.0172],[36,0.0182],[37,0.0167],[39,0.0177],[40,0.0175],[41,0.0189],[42,0.0198],[43,0.0185],[44,0.0179],[45,0.0185],[46,0.0159],[47,0.018],[48,0.0214],[49,0.0219],[50,0.0212],[51,0.0245],[52,0.0246],[53,0.02],[54,0.0189],[55,0.0171],[56,0.0174],[57,0.0162],[58,0.0159],[59,0.0174],[60,0.0177],[61,0.0173],[62,0.0177],[63,0.0176],[64,0.0164],[65,0.0164],[67,0.0151],[68,0.0161],[69,0.0159],[70,0.0165],[71,0.0155],[72,0.0156],[73,0.0145],[74,0.0146],[75,0.0159],[76,0.0164],[77,0.0152],[78,0.0152],[79,0.0138],[80,0.0134],[81,0.0146],[82,0.014],[83,0.0142],[84,0.0139],[85,0.0142],[86,0.0132],[88,0.0117],[89,0.0121],[90,0.0114],[91,0.0129],[94,0.0124],[95,0.0124],[96,0.0126],[97,0.0118],[100,0.0124],[102,0.0121],[103,0.0107],[104,0.0112],[105,0.0113],[106,0.0123],[107,0.0127],[108,0.0114],[109,0.0124],[110,0.0131],[113,0.0119],[114,0.0118],[115,0.0125],[116,0.0122],[117,0.0131],[118,0.0162],[119,0.0153],[120,0.0164],[121,0.0143],[122,0.0136],[123,0.0155],[124,0.0161],[125,0.0131],[126,0.014],[127,0.0136],[128,0.0143],[129,0.0126],[130,0.0125],[131,0.0116],[132,0.0119],[133,0.012],[134,0.0135],[135,0.0159],[136,0.0214],[137,0.0228],[138,0.0208],[139,0.0158],[140,0.0166],[141,0.0165],[142,0.019],[143,0.0204],[144,0.0232],[145,0.0229],[146,0.0271],[147,0.0282],[148,0.0243],[149,0.0243],[150,0.0233],[151,0.0251],[152,0.0232],[154,0.0234],[155,0.0231],[156,0.0231],[157,0.0216],[158,0.0209],[160,0.0216],[161,0.0234],[162,0.0203],[163,0.0201],[164,0.0212],[165,0.0199],[166,0.0174],[167,0.0191],[168,0.0232],[169,0.0241],[170,0.0227],[171,0.0209],[172,0.0222],[173,0.0221],[174,0.0287],[175,0.0308],[176,0.0291],[177,0.0292],[178,0.029],[179,0.0277],[180,0.0283],[181,0.0347],[182,0.0386],[183,0.0396],[184,0.0319],[185,0.03],[186,0.0352],[187,0.0381],[188,0.0384],[189,0.0384],[190,0.0377],[191,0.0352],[192,0.0371],[193,0.0308],[194,0.0284],[195,0.0282],[196,0.0328],[197,0.0416],[198,0.0518],[199,0.0648],[200,0.0687],[201,0.077],[202,0.061],[203,0.0485],[204,0.0414],[205,0.0394],[206,0.0397],[207,0.0395],[208,0.0376],[209,0.0361],[210,0.0325],[211,0.0297],[212,0.0296],[213,0.0305],[214,0.0307],[215,0.0301],[216,0.0318],[217,0.032],[218,0.0336],[219,0.0338],[220,0.0353],[221,0.0342],[222,0.0356],[223,0.0393],[224,0.0396],[225,0.0378],[226,0.0386],[227,0.041],[228,0.042],[229,0.0393],[230,0.041],[231,0.0419],[232,0.0446],[233,0.0507],[234,0.0527],[235,0.0617],[236,0.0642],[237,0.0722],[238,0.0723],[240,0.0769],[241,0.0895],[242,0.0925],[243,0.0894],[244,0.1057],[245,0.1117],[246,0.1386],[247,0.1554],[248,0.1525],[249,0.125],[250,0.1048],[251,0.1052],[252,0.0921],[253,0.0883],[254,0.0938],[255,0.0886],[256,0.0881],[257,0.0964],[258,0.1159],[259,0.1155],[260,0.1097],[261,0.0988],[262,0.0782],[263,0.0529],[264,0.0519],[265,0.0506],[266,0.05],[268,0.0441],[269,0.043],[270,0.0429],[271,0.0451],[272,0.0515],[273,0.0606],[274,0.0629],[275,0.0691],[276,0.0679],[277,0.0653],[278,0.0639],[279,0.0627],[280,0.0628],[281,0.0617],[282,0.0627],[283,0.0619],[284,0.0618],[285,0.0631],[286,0.0596],[288,0.0623],[289,0.0621],[290,0.0628],[291,0.0602],[292,0.0543],[293,0.0514],[294,0.0489],[295,0.0525],[296,0.0638],[297,0.0656],[298,0.0669],[299,0.0644],[300,0.0599],[301,0.0572],[302,0.036],[303,0.0339],[304,0.0278],[305,0.0255],[307,0.0249],[308,0.0238],[309,0.0247],[310,0.0254],[311,0.0302],[312,0.0309],[313,0.031],[314,0.0302],[316,0.0262],[317,0.0267],[318,0.0224],[319,0.0233],[320,0.026],[321,0.0252],[322,0.0253],[323,0.0277],[324,0.0293],[325,0.0285],[326,0.0331],[327,0.0356],[328,0.0386],[329,0.0374],[330,0.0407],[331,0.0393],[332,0.0371],[333,0.0367],[334,0.0351],[335,0.0377],[336,0.0366],[337,0.0339],[338,0.036],[339,0.0343],[340,0.0363],[341,0.0332],[342,0.0347],[343,0.0339],[344,0.0314],[346,0.0359],[347,0.0349],[348,0.0322],[349,0.0312],[350,0.0299],[351,0.0272],[352,0.0262],[353,0.0287],[354,0.0271],[355,0.0274],[356,0.0279],[357,0.0301],[358,0.0299],[359,0.0275],[360,0.0275]]
       }
       CTRL_L_mouth_lipBiteU {
        type=MovieSceneScriptingFloatChannel
        default=0
        thinned_from=492
        keys [[-131,0.0422],[-130,0.0414],[-129,0.0379],[-128,0.0364],[-127,0.0358],[-126,0.0356],[-125,0.0358],[-124,0.0357],[-123,0.0363],[-122,0.0365],[-121,0.0371],[-120,0.0371],[-119,0.0369],[-118,0.0364],[-117,0.0365],[-116,0.037],[-115,0.0351],[-114,0.0319],[-113,0.0294],[-112,0.0277],[-111,0.0266],[-110,0.026],[-109,0.0258],[-107,0.0244],[-106,0.0243],[-105,0.0239],[-104,0.0243],[-103,0.0238],[-102,0.0236],[-101,0.0215],[-100,0.0214],[-99,0.0231],[-98,0.023],[-96,0.0211],[-95,0.021],[-94,0.0255],[-93,0.0258],[-92,0.0268],[-91,0.0291],[-90,0.034],[-89,0.0432],[-88,0.0533],[-87,0.0676],[-86,0.0855],[-85,0.1059],[-84,0.1255],[-83,0.1391],[-82,0.1579],[-81,0.1801],[-80,0.2079],[-79,0.2119],[-78,0.2476],[-77,0.2881],[-76,0.3169],[-75,0.3589],[-74,0.3882],[-73,0.3989],[-72,0.4115],[-70,0.4229],[-69,0.4329],[-68,0.4318],[-67,0.4289],[-66,0.4303],[-65,0.4298],[-64,0.4254],[-63,0.4175],[-62,0.4143],[-61,0.4038],[-60,0.3755],[-59,0.3232],[-58,0.2234],[-57,0.2075],[-56,0.1692],[-55,0.142],[-54,0.1256],[-53,0.1146],[-52,0.1094],[-51,0.0902],[-50,0.0827],[-49,0.082],[-48,0.0749],[-47,0.0828],[-46,0.086],[-45,0.1011],[-44,0.1037],[-43,0.1138],[-42,0.1299],[-41,0.1465],[-40,0.171],[-39,0.207],[-38,0.2188],[-37,0.3113],[-36,0.3521],[-35,0.3789],[-34,0.3904],[-33,0.3772],[-32,0.3109],[-31,0.2843],[-30,0.1531],[-29,0.1449],[-28,0.1261],[-27,0.113],[-26,0.1058],[-25,0.1007],[-24,0.0992],[-23,0.0913],[-22,0.0724],[-21,0.0637],[-20,0.0764],[-19,0.0922],[-18,0.098],[-17,0.0942],[-16,0.0842],[-15,0.0784],[-14,0.0762],[-13,0.078],[-12,0.0762],[-11,0.0722],[-10,0.0732],[-9,0.0699],[-8,0.0643],[-7,0.0644],[-6,0.0639],[-5,0.0607],[-4,0.0586],[-3,0.058],[-2,0.058],[-1,0.0536],[0,0.052],[1,0.0524],[2,0.0518],[3,0.0522],[4,0.0522],[5,0.0539],[6,0.0533],[7,0.0535],[8,0.0529],[10,0.0487],[11,0.0493],[12,0.0523],[13,0.055],[14,0.0548],[15,0.0542],[16,0.0556],[17,0.0578],[18,0.0566],[19,0.0551],[20,0.0576],[21,0.059],[22,0.0584],[23,0.058],[24,0.0574],[26,0.0597],[27,0.0612],[28,0.0617],[30,0.0607],[31,0.0598],[32,0.0591],[33,0.0615],[34,0.0618],[35,0.0614],[36,0.0585],[38,0.0617],[39,0.0616],[41,0.0583],[42,0.0573],[43,0.0592],[44,0.0617],[45,0.062],[46,0.0534],[47,0.0547],[48,0.0605],[49,0.061],[50,0.0646],[51,0.0644],[52,0.0649],[53,0.0662],[54,0.0638],[55,0.0568],[56,0.0543],[57,0.0542],[58,0.0503],[59,0.0474],[61,0.0466],[62,0.0464],[63,0.0459],[64,0.0387],[65,0.0387],[66,0.039],[67,0.0419],[69,0.0407],[70,0.0406],[72,0.0373],[73,0.0396],[74,0.0407],[75,0.0437],[76,0.0446],[77,0.04],[78,0.0399],[79,0.0382],[80,0.0393],[81,0.0395],[83,0.0373],[84,0.0389],[85,0.0391],[86,0.0354],[87,0.0349],[88,0.0366],[90,0.0354],[91,0.0365],[92,0.038],[93,0.0376],[95,0.0351],[96,0.0365],[97,0.0373],[98,0.0372],[99,0.0373],[100,0.0356],[101,0.035],[102,0.035],[103,0.0344],[104,0.0341],[105,0.0341],[106,0.0337],[108,0.0334],[109,0.0338],[110,0.0354],[111,0.0357],[112,0.0342],[113,0.0313],[114,0.0296],[115,0.0307],[116,0.0321],[117,0.0318],[119,0.0357],[120,0.037],[121,0.0359],[122,0.0334],[123,0.0329],[124,0.033],[125,0.0328],[126,0.0325],[127,0.0316],[128,0.0306],[129,0.0303],[130,0.0284],[132,0.0275],[133,0.026],[134,0.0264],[135,0.0281],[137,0.0287],[138,0.0265],[139,0.0228],[140,0.0234],[142,0.0251],[143,0.0251],[144,0.0247],[145,0.0248],[147,0.0239],[148,0.0242],[149,0.0252],[150,0.0255],[151,0.0249],[153,0.0248],[154,0.0249],[155,0.0254],[156,0.0249],[157,0.0238],[158,0.0236],[159,0.0244],[161,0.0241],[162,0.0245],[164,0.024],[165,0.024],[167,0.0253],[168,0.0266],[169,0.0311],[170,0.0333],[171,0.0357],[172,0.0377],[174,0.0375],[175,0.0376],[176,0.0407],[178,0.0412],[179,0.0398],[180,0.0397],[181,0.0411],[182,0.0417],[183,0.04],[184,0.0381],[185,0.0364],[186,0.0362],[187,0.034],[188,0.0335],[189,0.0312],[190,0.0305],[191,0.03],[192,0.031],[193,0.029],[194,0.0287],[195,0.0312],[196,0.0409],[197,0.0491],[198,0.0529],[199,0.0584],[200,0.0603],[201,0.0613],[202,0.0515],[203,0.0476],[204,0.052],[205,0.0476],[206,0.0455],[207,0.0453],[208,0.0447],[209,0.045],[210,0.0443],[211,0.0423],[212,0.042],[213,0.0422],[214,0.0409],[215,0.0405],[216,0.0429],[217,0.0431],[218,0.0423],[219,0.0433],[220,0.0448],[221,0.0449],[222,0.046],[223,0.0483],[224,0.0489],[225,0.0477],[226,0.0479],[228,0.049],[229,0.0492],[230,0.0505],[231,0.0525],[232,0.0523],[234,0.0535],[235,0.0508],[236,0.0476],[237,0.0472],[238,0.0486],[239,0.0571],[240,0.0604],[241,0.0631],[242,0.0637],[244,0.0629],[245,0.065],[246,0.0723],[247,0.0805],[248,0.0815],[249,0.0755],[250,0.068],[251,0.0666],[252,0.0677],[253,0.0704],[254,0.0749],[255,0.0712],[256,0.0653],[257,0.0624],[258,0.053],[260,0.0527],[261,0.057],[262,0.0576],[263,0.0546],[264,0.057],[265,0.0607],[266,0.0592],[267,0.0546],[268,0.0472],[269,0.0429],[270,0.0425],[271,0.0508],[272,0.0517],[273,0.0492],[274,0.0476],[276,0.0471],[277,0.0467],[278,0.047],[283,0.0459],[284,0.0459],[285,0.0455],[288,0.0477],[289,0.0476],[290,0.0472],[291,0.0456],[292,0.0413],[293,0.0359],[294,0.0347],[295,0.0392],[296,0.0462],[297,0.0474],[298,0.0521],[299,0.051],[300,0.0399],[301,0.0392],[302,0.0352],[303,0.0349],[304,0.0344],[305,0.0346],[307,0.0372],[308,0.0378],[309,0.0371],[310,0.0369],[311,0.039],[312,0.0391],[313,0.0376],[314,0.0364],[317,0.0358],[318,0.0358],[319,0.0355],[320,0.0361],[321,0.037],[322,0.0359],[323,0.0352],[324,0.0354],[325,0.0353],[326,0.0362],[327,0.0377],[328,0.0378],[329,0.0376],[330,0.0385],[331,0.0389],[332,0.0378],[333,0.0378],[334,0.0375],[335,0.0375],[336,0.0383],[337,0.038],[338,0.0384],[339,0.0393],[340,0.0397],[341,0.0394],[342,0.0395],[343,0.04],[344,0.0391],[345,0.0398],[346,0.0402],[347,0.0386],[349,0.0394],[350,0.039],[352,0.0393],[353,0.0406],[354,0.0402],[355,0.0395],[356,0.0394],[358,0.0398],[359,0.0392],[360,0.0392]]
       }
       CTRL_R_mouth_pressD {type=MovieSceneScriptingFloatChannel, default=0, const [0,492]}
       CTRL_R_mouth_pressU {type=MovieSceneScriptingFloatChannel, default=0, const [0,492]}
       CTRL_L_mouth_purseD {
        type=MovieSceneScriptingFloatChannel
        default=0
        thinned_from=492
        keys [[-131,0.0843],[-130,0.0834],[-129,0.0753],[-128,0.0699],[-127,0.0683],[-126,0.0671],[-125,0.0676],[-124,0.0668],[-122,0.0671],[-121,0.0675],[-120,0.0693],[-119,0.0702],[-118,0.0683],[-117,0.0674],[-116,0.0608],[-115,0.0587],[-113,0.0565],[-111,0.0528],[-110,0.0527],[-109,0.0521],[-108,0.0499],[-107,0.0469],[-106,0.0464],[-105,0.0452],[-104,0.0476],[-103,0.0475],[-102,0.0477],[-101,0.0494],[-100,0.0466],[-99,0.0498],[-98,0.0494],[-97,0.0485],[-95,0.0495],[-94,0.0574],[-92,0.0582],[-91,0.0626],[-90,0.0683],[-89,0.0731],[-88,0.082],[-87,0.0979],[-86,0.1102],[-85,0.1242],[-84,0.1401],[-83,0.1613],[-82,0.2134],[-81,0.2658],[-80,0.3071],[-79,0.3123],[-78,0.3572],[-77,0.3613],[-76,0.3698],[-75,0.373],[-74,0.3702],[-72,0.3801],[-71,0.3751],[-70,0.3735],[-69,0.371],[-67,0.3575],[-66,0.364],[-65,0.3596],[-64,0.353],[-62,0.3295],[-61,0.3111],[-60,0.2696],[-59,0.205],[-58,0.13],[-57,0.1127],[-56,0.0758],[-55,0.0582],[-54,0.0486],[-53,0.0432],[-52,0.0407],[-51,0.0391],[-50,0.0381],[-49,0.0379],[-48,0.0433],[-47,0.0474],[-46,0.0507],[-45,0.0703],[-44,0.0779],[-43,0.1065],[-42,0.1424],[-41,0.1766],[-40,0.2383],[-39,0.301],[-38,0.3159],[-37,0.3361],[-36,0.33],[-35,0.3288],[-34,0.3143],[-33,0.2789],[-32,0.2005],[-31,0.1801],[-30,0.0892],[-29,0.0805],[-28,0.0612],[-27,0.0498],[-26,0.0445],[-25,0.0401],[-24,0.0388],[-23,0.0423],[-22,0.0508],[-21,0.0571],[-20,0.0522],[-19,0.0532],[-18,0.0597],[-17,0.0519],[-16,0.0483],[-15,0.0482],[-14,0.0533],[-13,0.0517],[-12,0.0481],[-11,0.0458],[-10,0.0422],[-8,0.0406],[-7,0.0368],[-6,0.0353],[-4,0.0363],[-3,0.0366],[-2,0.0339],[-1,0.0355],[0,0.0343],[1,0.0344],[3,0.036],[4,0.0371],[5,0.0372],[6,0.0381],[8,0.0386],[10,0.0398],[12,0.0381],[13,0.0384],[14,0.0394],[15,0.04],[16,0.0413],[17,0.0413],[18,0.0394],[19,0.0404],[20,0.044],[21,0.0415],[22,0.0409],[25,0.0407],[27,0.0401],[28,0.0401],[29,0.0415],[30,0.0409],[31,0.0394],[32,0.0411],[33,0.0413],[34,0.0408],[35,0.0411],[36,0.0428],[37,0.0409],[39,0.0389],[40,0.0398],[41,0.0418],[42,0.042],[43,0.0407],[44,0.0404],[45,0.0411],[46,0.0399],[47,0.0413],[48,0.0442],[49,0.0452],[50,0.045],[51,0.0474],[52,0.0472],[53,0.041],[55,0.0387],[56,0.0396],[58,0.038],[59,0.0391],[61,0.0396],[62,0.0393],[63,0.0393],[64,0.0384],[65,0.0384],[66,0.0377],[67,0.0373],[68,0.0395],[69,0.0392],[70,0.0394],[71,0.0392],[72,0.0387],[73,0.0379],[74,0.0378],[75,0.0393],[76,0.0387],[77,0.037],[78,0.0369],[79,0.0355],[80,0.0356],[81,0.0361],[82,0.0355],[84,0.0353],[85,0.0356],[86,0.0342],[88,0.0332],[89,0.033],[90,0.0327],[91,0.0343],[92,0.0339],[95,0.0343],[96,0.034],[97,0.0329],[98,0.0327],[99,0.033],[100,0.0335],[101,0.0336],[102,0.0334],[103,0.0318],[105,0.0319],[107,0.0332],[108,0.0322],[109,0.0325],[112,0.0325],[113,0.0328],[114,0.0329],[115,0.0338],[116,0.0339],[117,0.0347],[118,0.0379],[119,0.0372],[120,0.0383],[121,0.0364],[122,0.0353],[123,0.0367],[124,0.037],[125,0.0348],[126,0.0353],[127,0.035],[128,0.0357],[129,0.0342],[130,0.0341],[131,0.0329],[133,0.0329],[134,0.0362],[135,0.04],[136,0.0446],[137,0.0468],[138,0.0432],[139,0.0338],[140,0.0322],[141,0.0316],[142,0.0345],[144,0.0408],[145,0.0408],[146,0.046],[147,0.0467],[148,0.0432],[149,0.0427],[150,0.0417],[151,0.0428],[152,0.0411],[154,0.0411],[155,0.0423],[156,0.0412],[157,0.0385],[158,0.0385],[159,0.0392],[160,0.0393],[161,0.0409],[162,0.0367],[163,0.0378],[164,0.0395],[165,0.0383],[166,0.0352],[167,0.0372],[168,0.0425],[169,0.0469],[171,0.0453],[172,0.0472],[173,0.0482],[174,0.0548],[175,0.0564],[177,0.0556],[179,0.0514],[180,0.0515],[181,0.0574],[182,0.061],[183,0.0617],[184,0.0544],[185,0.0526],[186,0.0574],[187,0.0611],[190,0.0624],[191,0.0597],[192,0.0611],[193,0.0556],[194,0.0548],[195,0.0549],[196,0.0592],[197,0.0666],[198,0.0755],[199,0.087],[200,0.0904],[201,0.0969],[202,0.0845],[203,0.0689],[204,0.0606],[205,0.0593],[206,0.0597],[207,0.0596],[208,0.0578],[209,0.0567],[210,0.0536],[211,0.0516],[212,0.0518],[213,0.0529],[214,0.0533],[215,0.0528],[216,0.0542],[217,0.0549],[218,0.0564],[219,0.0564],[220,0.0574],[221,0.0562],[222,0.0573],[223,0.0601],[224,0.0602],[225,0.0588],[226,0.06],[227,0.0619],[228,0.0633],[229,0.0608],[231,0.0639],[232,0.0667],[233,0.0727],[234,0.074],[235,0.0833],[236,0.0904],[237,0.1002],[238,0.0983],[240,0.0965],[241,0.1064],[242,0.1089],[243,0.1069],[244,0.1208],[245,0.1262],[246,0.1522],[247,0.1715],[248,0.1698],[249,0.1456],[250,0.1257],[251,0.1259],[252,0.1094],[253,0.1049],[254,0.1113],[255,0.1076],[256,0.1068],[257,0.1144],[258,0.1354],[259,0.1346],[260,0.1255],[261,0.1137],[262,0.0932],[263,0.0687],[265,0.0651],[266,0.0641],[267,0.0618],[268,0.0621],[269,0.0601],[270,0.0596],[271,0.0614],[272,0.0676],[273,0.0757],[274,0.0758],[275,0.0813],[276,0.0799],[277,0.0771],[278,0.0752],[279,0.0736],[280,0.0736],[281,0.0725],[282,0.0741],[283,0.0734],[284,0.0733],[285,0.0745],[286,0.0715],[288,0.0739],[289,0.074],[290,0.0747],[291,0.0725],[292,0.0678],[293,0.067],[294,0.065],[295,0.0683],[296,0.0821],[297,0.0845],[298,0.088],[299,0.087],[300,0.0849],[301,0.0823],[302,0.0612],[303,0.0591],[304,0.0526],[305,0.05],[306,0.0496],[307,0.0488],[308,0.0475],[309,0.0484],[310,0.049],[311,0.0526],[313,0.0541],[314,0.0535],[315,0.0506],[316,0.0488],[317,0.0494],[318,0.0454],[319,0.0461],[320,0.0487],[322,0.0483],[323,0.0503],[324,0.0521],[325,0.0513],[326,0.0548],[328,0.0596],[329,0.0581],[330,0.061],[331,0.0596],[332,0.0572],[334,0.0554],[335,0.0578],[336,0.0563],[337,0.0541],[338,0.0561],[339,0.0545],[340,0.0569],[341,0.0541],[342,0.0553],[343,0.0545],[344,0.0519],[345,0.0542],[346,0.0569],[347,0.0561],[348,0.0532],[349,0.0526],[350,0.0515],[351,0.0496],[352,0.0487],[353,0.0517],[354,0.0504],[355,0.0505],[356,0.051],[357,0.0528],[358,0.0529],[359,0.0506],[360,0.0506]]
       }
       CTRL_R_mouth_lipBiteU {
        type=MovieSceneScriptingFloatChannel
        default=0
        thinned_from=492
        keys [[-131,0.0285],[-130,0.0266],[-129,0.0234],[-128,0.0223],[-127,0.0209],[-126,0.0207],[-125,0.0208],[-124,0.02],[-122,0.02],[-121,0.0197],[-120,0.021],[-118,0.0199],[-117,0.0199],[-116,0.0195],[-115,0.0184],[-114,0.0169],[-113,0.0152],[-112,0.0142],[-110,0.0131],[-108,0.0127],[-106,0.0129],[-102,0.0119],[-101,0.0112],[-100,0.0108],[-97,0.0112],[-96,0.0109],[-95,0.0108],[-94,0.012],[-93,0.0121],[-92,0.0124],[-91,0.0135],[-90,0.0162],[-89,0.0231],[-88,0.032],[-87,0.046],[-86,0.064],[-85,0.0842],[-84,0.1032],[-83,0.1173],[-82,0.137],[-81,0.1602],[-80,0.189],[-79,0.1925],[-78,0.2244],[-77,0.2603],[-76,0.2866],[-75,0.326],[-74,0.3542],[-73,0.3649],[-72,0.3786],[-71,0.3875],[-70,0.3935],[-69,0.4012],[-68,0.399],[-67,0.3949],[-66,0.3966],[-65,0.3963],[-64,0.3928],[-63,0.3869],[-62,0.383],[-61,0.3721],[-60,0.3426],[-59,0.2889],[-58,0.1898],[-57,0.1751],[-56,0.1395],[-55,0.1131],[-54,0.0975],[-53,0.0889],[-52,0.085],[-51,0.0666],[-50,0.0596],[-49,0.059],[-48,0.054],[-47,0.0616],[-46,0.0646],[-45,0.0787],[-44,0.0814],[-43,0.091],[-42,0.1054],[-41,0.1215],[-40,0.146],[-39,0.1804],[-38,0.1913],[-37,0.2807],[-36,0.3209],[-35,0.3471],[-34,0.3548],[-33,0.3403],[-32,0.2778],[-31,0.2529],[-30,0.1288],[-29,0.1203],[-28,0.101],[-27,0.0884],[-26,0.0814],[-25,0.0761],[-24,0.0746],[-23,0.0667],[-22,0.0476],[-21,0.0382],[-20,0.052],[-19,0.0683],[-18,0.0725],[-17,0.0671],[-16,0.0586],[-15,0.0534],[-14,0.0498],[-13,0.0513],[-12,0.0494],[-11,0.0463],[-10,0.0473],[-9,0.0441],[-8,0.0399],[-7,0.04],[-6,0.0396],[-5,0.0382],[-4,0.0373],[-2,0.0367],[-1,0.0331],[0,0.0318],[2,0.032],[3,0.0326],[4,0.0329],[5,0.034],[6,0.0348],[7,0.0349],[8,0.0345],[10,0.0316],[11,0.032],[12,0.0334],[13,0.0346],[14,0.0344],[15,0.0346],[16,0.0359],[17,0.037],[18,0.0359],[19,0.0352],[20,0.0371],[21,0.0378],[22,0.0373],[23,0.037],[24,0.0366],[26,0.038],[28,0.0402],[29,0.0404],[30,0.0399],[31,0.0387],[32,0.0379],[33,0.0397],[35,0.0393],[36,0.0377],[37,0.0392],[38,0.0398],[40,0.0378],[41,0.0373],[42,0.0364],[43,0.0373],[44,0.0395],[45,0.04],[46,0.0359],[47,0.0365],[48,0.0386],[49,0.0395],[50,0.0427],[51,0.0417],[52,0.042],[53,0.0438],[54,0.0425],[55,0.0381],[56,0.0359],[57,0.035],[58,0.0313],[59,0.0279],[60,0.0275],[61,0.0274],[62,0.027],[63,0.0267],[64,0.0226],[65,0.0226],[66,0.023],[67,0.0246],[69,0.025],[70,0.025],[71,0.0243],[72,0.0232],[74,0.0259],[75,0.028],[76,0.0291],[77,0.0262],[78,0.0262],[79,0.0253],[80,0.0264],[81,0.0271],[83,0.0252],[84,0.026],[85,0.0263],[86,0.0242],[87,0.0232],[88,0.0238],[90,0.023],[91,0.0236],[92,0.0245],[93,0.0243],[94,0.0234],[95,0.0228],[96,0.0236],[97,0.024],[98,0.0234],[99,0.0233],[100,0.0229],[102,0.0228],[103,0.0211],[104,0.0216],[105,0.0216],[107,0.021],[109,0.0216],[111,0.0218],[112,0.0215],[113,0.02],[114,0.0191],[115,0.0201],[116,0.0213],[117,0.021],[120,0.0238],[121,0.0243],[122,0.0236],[123,0.0225],[124,0.0222],[125,0.0229],[126,0.0228],[128,0.0219],[129,0.022],[130,0.0206],[132,0.0187],[133,0.0183],[134,0.0208],[135,0.0235],[136,0.0241],[137,0.0253],[138,0.025],[139,0.0199],[140,0.0181],[141,0.0189],[142,0.0201],[143,0.0207],[145,0.02],[146,0.02],[147,0.0198],[148,0.02],[149,0.0208],[150,0.0212],[152,0.0208],[153,0.021],[154,0.021],[155,0.0215],[156,0.021],[157,0.0197],[158,0.0197],[159,0.0206],[161,0.0192],[162,0.0194],[166,0.0198],[167,0.02],[168,0.0212],[169,0.027],[170,0.0293],[171,0.031],[172,0.0319],[174,0.0305],[175,0.0291],[176,0.0302],[177,0.0303],[178,0.0298],[179,0.0276],[180,0.0266],[181,0.0272],[182,0.027],[183,0.0245],[184,0.0225],[185,0.0214],[186,0.0211],[187,0.02],[188,0.0198],[189,0.0188],[190,0.0186],[191,0.0187],[192,0.0198],[193,0.02],[194,0.0212],[195,0.0218],[196,0.0258],[197,0.0302],[198,0.0334],[199,0.0382],[200,0.0398],[201,0.0412],[202,0.0335],[203,0.027],[204,0.0273],[205,0.0249],[206,0.024],[207,0.0236],[208,0.0234],[209,0.0236],[211,0.0221],[212,0.0222],[213,0.0227],[214,0.0221],[215,0.0219],[216,0.0236],[217,0.0239],[218,0.0238],[222,0.026],[223,0.0271],[224,0.0272],[225,0.0266],[226,0.027],[227,0.027],[229,0.0277],[230,0.0291],[231,0.0308],[232,0.0309],[233,0.0319],[234,0.0319],[235,0.0302],[236,0.0322],[237,0.0326],[238,0.0306],[239,0.0337],[240,0.0364],[241,0.0386],[242,0.0391],[243,0.038],[244,0.0385],[245,0.0402],[246,0.0447],[247,0.0526],[248,0.0572],[249,0.0558],[250,0.0502],[251,0.0484],[252,0.0487],[253,0.0515],[254,0.0573],[255,0.0553],[256,0.0496],[257,0.0468],[258,0.0391],[259,0.0389],[260,0.0382],[261,0.0398],[263,0.0368],[264,0.0386],[265,0.0416],[266,0.0404],[267,0.0364],[268,0.03],[269,0.0265],[270,0.0262],[271,0.0315],[272,0.0312],[273,0.0295],[274,0.0282],[275,0.0281],[276,0.0275],[277,0.0266],[278,0.0266],[280,0.0254],[282,0.0257],[285,0.0255],[288,0.0258],[290,0.0252],[291,0.0235],[293,0.0179],[294,0.0171],[295,0.0201],[296,0.0261],[297,0.0272],[298,0.0332],[299,0.0329],[300,0.0254],[301,0.025],[302,0.0219],[303,0.0216],[304,0.0209],[305,0.0208],[307,0.0216],[308,0.0217],[309,0.0213],[310,0.0211],[311,0.0217],[312,0.0217],[313,0.0211],[315,0.0209],[320,0.0211],[321,0.0215],[323,0.0204],[325,0.0212],[326,0.0215],[327,0.0221],[329,0.0224],[330,0.023],[331,0.0234],[332,0.0225],[333,0.0221],[335,0.0228],[336,0.0229],[337,0.0224],[339,0.0234],[342,0.0236],[343,0.024],[344,0.0232],[345,0.0233],[346,0.0236],[347,0.0231],[348,0.0232],[350,0.0227],[351,0.0227],[352,0.0229],[353,0.0241],[354,0.024],[355,0.0233],[358,0.0231],[359,0.0226],[360,0.0226]]
       }
       CTRL_R_mouth_suckBlow {
        type=MovieSceneScriptingFloatChannel
        default=0
        thinned_from=492
        keys [[-131,0.0059],[-92,0.0059],[-91,0.006],[-89,0.0068],[-87,0.0064],[-85,0.0068],[-61,0.0069],[-60,0.0075],[-59,0.013],[-58,0.0288],[-57,0.0326],[-56,0.0409],[-55,0.0443],[-54,0.0449],[-52,0.0438],[-51,0.0389],[-50,0.0333],[-49,0.0327],[-48,0.0286],[-47,0.023],[-46,0.0211],[-45,0.0123],[-44,0.0113],[-43,0.0082],[-42,0.0068],[-33,0.0069],[-32,0.0078],[-31,0.0111],[-30,0.0321],[-29,0.0328],[-28,0.0343],[-27,0.0363],[-26,0.0368],[-25,0.0377],[-24,0.038],[-23,0.032],[-22,0.0184],[-21,0.0124],[-20,0.0118],[-19,0.0132],[-18,0.0117],[-17,0.0112],[-16,0.0117],[-13,0.0117],[-11,0.0137],[-10,0.0133],[-7,0.0145],[-6,0.0156],[-5,0.0161],[-4,0.0156],[-3,0.0157],[-2,0.0171],[-1,0.018],[0,0.0171],[1,0.0176],[2,0.019],[3,0.019],[5,0.0171],[6,0.0166],[10,0.0166],[12,0.0163],[14,0.0176],[15,0.0176],[16,0.0172],[17,0.0171],[18,0.0185],[19,0.0195],[20,0.0195],[21,0.02],[23,0.0193],[24,0.0185],[25,0.0182],[27,0.0198],[28,0.02],[30,0.0187],[32,0.019],[33,0.0186],[35,0.0187],[36,0.0195],[37,0.0195],[39,0.0203],[40,0.0193],[41,0.0193],[42,0.0211],[43,0.0212],[44,0.0199],[45,0.0195],[46,0.0199],[47,0.0181],[48,0.0143],[49,0.0132],[51,0.0095],[52,0.0107],[53,0.0146],[54,0.0146],[55,0.014],[56,0.0119],[57,0.0107],[58,0.0104],[59,0.0083],[61,0.0073],[63,0.0074],[64,0.0078],[66,0.0078],[67,0.0083],[69,0.0083],[72,0.0098],[73,0.0108],[74,0.0124],[75,0.0127],[76,0.0128],[77,0.0142],[78,0.0142],[79,0.0151],[80,0.0151],[81,0.0156],[82,0.017],[83,0.019],[84,0.0167],[85,0.0162],[86,0.019],[87,0.0175],[88,0.0157],[89,0.0152],[91,0.0152],[92,0.0166],[93,0.017],[94,0.0178],[96,0.0153],[97,0.0155],[98,0.0144],[99,0.0139],[100,0.0144],[101,0.0136],[102,0.0132],[103,0.0136],[105,0.0137],[106,0.0135],[108,0.0126],[110,0.0108],[112,0.0107],[113,0.0093],[114,0.0088],[115,0.0102],[116,0.0098],[117,0.0097],[118,0.0093],[119,0.0084],[120,0.0088],[122,0.0098],[123,0.0098],[124,0.0103],[127,0.0131],[128,0.0137],[129,0.0146],[130,0.0166],[131,0.0166],[132,0.0157],[133,0.0171],[134,0.0196],[135,0.0205],[136,0.0173],[137,0.021],[138,0.0354],[139,0.0312],[140,0.021],[141,0.0205],[142,0.0191],[143,0.0183],[144,0.0185],[145,0.0165],[146,0.0133],[147,0.0132],[148,0.0137],[150,0.0137],[151,0.0142],[152,0.0156],[154,0.0147],[155,0.0156],[157,0.0156],[158,0.0161],[159,0.0171],[160,0.0165],[161,0.0151],[163,0.0137],[164,0.0128],[165,0.0132],[166,0.0126],[168,0.0106],[169,0.009],[170,0.0085],[171,0.0076],[172,0.0073],[173,0.0077],[175,0.0068],[189,0.0068],[191,0.0059],[192,0.006],[194,0.0068],[195,0.0066],[196,0.0061],[197,0.0059],[199,0.0067],[236,0.0068],[237,0.0066],[238,0.0061],[239,0.0059],[240,0.0062],[241,0.0067],[245,0.0068],[247,0.0059],[259,0.0059],[261,0.0063],[265,0.0063],[267,0.0059],[297,0.0059],[299,0.0063],[301,0.0064],[302,0.0068],[305,0.0068],[306,0.0071],[308,0.0081],[310,0.0083],[311,0.0078],[312,0.0078],[313,0.0083],[314,0.0098],[315,0.0102],[316,0.0098],[319,0.0098],[321,0.0088],[322,0.0093],[323,0.0093],[324,0.0088],[327,0.0088],[328,0.0093],[329,0.0092],[330,0.0082],[331,0.0078],[333,0.0087],[334,0.0087],[335,0.0083],[336,0.0083],[337,0.0088],[339,0.0088],[341,0.0078],[342,0.0083],[343,0.0083],[344,0.0078],[351,0.0078],[353,0.0088],[357,0.0088],[358,0.0093],[360,0.0092]]
       }
       CTRL_L_mouth_purseU {
        type=MovieSceneScriptingFloatChannel
        default=0
        thinned_from=492
        keys [[-131,0.0712],[-130,0.0694],[-129,0.0622],[-128,0.0565],[-127,0.0565],[-126,0.0576],[-125,0.0638],[-124,0.0663],[-123,0.067],[-121,0.0674],[-120,0.0719],[-119,0.0735],[-118,0.0707],[-117,0.0696],[-116,0.0623],[-115,0.0614],[-114,0.0616],[-113,0.0613],[-112,0.0603],[-111,0.0587],[-110,0.0577],[-109,0.0562],[-108,0.0504],[-107,0.0468],[-106,0.0463],[-105,0.0445],[-104,0.0446],[-103,0.0431],[-102,0.0439],[-101,0.0487],[-100,0.0461],[-99,0.0486],[-98,0.0497],[-97,0.0495],[-96,0.0487],[-95,0.0488],[-94,0.058],[-93,0.0581],[-92,0.0549],[-90,0.0625],[-89,0.0673],[-88,0.078],[-87,0.0975],[-86,0.1124],[-85,0.127],[-84,0.144],[-83,0.1695],[-82,0.2237],[-81,0.2768],[-80,0.3255],[-79,0.3334],[-78,0.4017],[-77,0.4281],[-76,0.4567],[-75,0.4867],[-74,0.4844],[-73,0.4975],[-72,0.5031],[-71,0.4967],[-69,0.4989],[-68,0.4887],[-67,0.4913],[-66,0.498],[-65,0.4947],[-64,0.49],[-63,0.4818],[-62,0.4751],[-61,0.4578],[-60,0.4081],[-59,0.3141],[-58,0.2103],[-57,0.1869],[-56,0.1377],[-55,0.114],[-54,0.0977],[-53,0.0873],[-52,0.0824],[-51,0.0768],[-50,0.0741],[-49,0.0737],[-48,0.0726],[-47,0.0775],[-46,0.0811],[-45,0.0996],[-44,0.1087],[-43,0.1438],[-42,0.1871],[-41,0.226],[-40,0.2963],[-39,0.3726],[-38,0.3924],[-37,0.4546],[-36,0.4585],[-35,0.4677],[-34,0.4544],[-33,0.4124],[-32,0.3083],[-31,0.2791],[-30,0.1479],[-29,0.1349],[-28,0.1068],[-27,0.0922],[-26,0.0843],[-25,0.076],[-24,0.0734],[-23,0.073],[-22,0.0776],[-21,0.0804],[-20,0.0755],[-19,0.0696],[-18,0.0718],[-17,0.0621],[-16,0.0536],[-15,0.0519],[-14,0.0576],[-13,0.0563],[-12,0.0528],[-10,0.0494],[-9,0.0504],[-8,0.0491],[-7,0.0435],[-5,0.0417],[-4,0.0399],[-3,0.0398],[-2,0.0385],[-1,0.0379],[0,0.0375],[1,0.0392],[2,0.0385],[4,0.0386],[5,0.0416],[6,0.0451],[7,0.047],[8,0.0474],[9,0.0483],[10,0.0481],[11,0.0466],[12,0.0466],[13,0.0477],[14,0.0496],[15,0.0502],[16,0.05],[17,0.0517],[18,0.0512],[19,0.053],[20,0.0574],[22,0.0549],[24,0.0545],[25,0.0552],[26,0.0549],[27,0.0549],[28,0.0546],[29,0.0565],[30,0.055],[31,0.0531],[32,0.0559],[33,0.0554],[34,0.0537],[35,0.0542],[36,0.0569],[37,0.0548],[38,0.0538],[39,0.0539],[40,0.0551],[41,0.0567],[42,0.0571],[43,0.0558],[44,0.0555],[45,0.0561],[46,0.0544],[47,0.0552],[48,0.0612],[49,0.063],[50,0.0593],[51,0.0644],[52,0.0686],[53,0.0696],[54,0.0687],[55,0.0651],[56,0.0633],[57,0.0626],[58,0.06],[59,0.0567],[60,0.0561],[62,0.057],[63,0.0567],[64,0.0533],[66,0.0531],[68,0.0547],[69,0.0567],[70,0.0568],[71,0.0558],[72,0.0568],[73,0.0555],[75,0.0633],[76,0.0643],[77,0.0618],[78,0.0617],[79,0.0595],[80,0.0586],[81,0.0604],[82,0.0582],[83,0.0576],[84,0.0579],[85,0.0587],[86,0.0533],[87,0.0519],[88,0.0491],[89,0.0498],[90,0.0487],[91,0.0513],[92,0.0529],[93,0.0528],[94,0.0517],[95,0.0504],[96,0.0511],[97,0.0487],[98,0.0478],[99,0.0479],[100,0.0492],[101,0.0491],[102,0.0484],[103,0.045],[104,0.0466],[105,0.0467],[106,0.0481],[107,0.0478],[108,0.0459],[109,0.0488],[110,0.0524],[111,0.0527],[112,0.0532],[113,0.0519],[114,0.0512],[115,0.0546],[116,0.054],[117,0.0568],[118,0.0635],[119,0.0622],[120,0.0632],[121,0.0603],[122,0.0595],[123,0.0637],[124,0.0652],[125,0.0616],[126,0.0644],[127,0.0608],[128,0.0621],[129,0.0574],[130,0.0563],[131,0.0557],[132,0.0547],[133,0.0507],[134,0.0494],[135,0.0512],[136,0.0577],[137,0.0514],[138,0.0444],[139,0.0433],[140,0.0469],[141,0.0493],[142,0.0542],[143,0.0551],[144,0.0557],[145,0.0547],[146,0.0591],[147,0.0617],[148,0.0586],[149,0.0561],[150,0.0551],[151,0.0563],[152,0.0565],[153,0.0559],[154,0.0561],[155,0.0555],[157,0.0519],[158,0.0504],[159,0.0533],[160,0.0527],[161,0.0535],[162,0.0498],[163,0.0485],[164,0.0509],[165,0.0494],[166,0.0461],[167,0.0499],[168,0.0546],[169,0.0539],[170,0.0527],[171,0.0483],[172,0.0509],[173,0.053],[174,0.0582],[175,0.0588],[176,0.0577],[177,0.0576],[178,0.0563],[179,0.0527],[180,0.0529],[181,0.061],[182,0.0645],[183,0.0649],[184,0.0557],[185,0.052],[187,0.061],[188,0.0616],[189,0.0599],[190,0.0588],[191,0.0554],[192,0.0593],[193,0.0534],[194,0.0498],[195,0.0503],[196,0.0487],[197,0.051],[198,0.0576],[199,0.0695],[200,0.0736],[201,0.0821],[202,0.0578],[203,0.048],[204,0.044],[205,0.044],[206,0.0435],[207,0.0413],[208,0.0429],[209,0.0422],[210,0.0409],[211,0.0418],[212,0.0411],[214,0.0462],[215,0.0446],[218,0.0506],[219,0.0515],[220,0.0531],[221,0.0537],[223,0.0578],[224,0.0586],[225,0.0567],[226,0.0581],[227,0.0605],[228,0.0626],[229,0.0559],[230,0.0559],[231,0.0554],[232,0.0562],[233,0.056],[234,0.0551],[237,0.087],[238,0.0849],[239,0.0874],[240,0.0942],[241,0.1067],[242,0.1092],[243,0.1053],[244,0.1243],[245,0.128],[246,0.1497],[247,0.1617],[248,0.1592],[249,0.1323],[250,0.1119],[251,0.1134],[252,0.1011],[253,0.0965],[254,0.1049],[255,0.1016],[256,0.1006],[257,0.1084],[258,0.132],[259,0.1309],[260,0.119],[261,0.1042],[262,0.0832],[263,0.0592],[264,0.0574],[265,0.0563],[266,0.0549],[267,0.0531],[268,0.0526],[269,0.0477],[270,0.0467],[271,0.0478],[272,0.0566],[273,0.0661],[274,0.0643],[275,0.0707],[276,0.0696],[277,0.0675],[278,0.068],[280,0.0694],[281,0.0679],[282,0.0702],[283,0.0693],[284,0.0692],[285,0.0712],[286,0.062],[287,0.0633],[288,0.065],[289,0.0651],[290,0.0658],[291,0.0624],[292,0.0563],[293,0.0536],[294,0.0505],[295,0.0521],[296,0.0626],[297,0.0647],[298,0.0744],[299,0.0773],[300,0.0796],[301,0.0772],[302,0.0578],[303,0.0557],[304,0.0502],[305,0.0495],[306,0.0503],[307,0.0501],[308,0.0492],[309,0.0509],[310,0.0518],[311,0.0566],[312,0.0578],[313,0.0596],[314,0.059],[315,0.0568],[316,0.0533],[317,0.0537],[318,0.0468],[319,0.0493],[320,0.0529],[321,0.0522],[322,0.053],[323,0.0552],[324,0.0568],[325,0.0562],[326,0.063],[327,0.0669],[328,0.0705],[329,0.0694],[330,0.0721],[331,0.0712],[332,0.0692],[334,0.0671],[335,0.0701],[336,0.0697],[337,0.0669],[338,0.0694],[339,0.0686],[340,0.0704],[341,0.0646],[342,0.0667],[343,0.0655],[344,0.0634],[345,0.0657],[346,0.069],[347,0.0673],[348,0.0636],[349,0.0628],[350,0.0605],[351,0.0574],[352,0.0565],[353,0.0606],[354,0.0583],[355,0.0586],[356,0.0593],[357,0.0621],[358,0.0617],[359,0.0574],[360,0.0574]]
       }
       CTRL_R_mouth_towardsD {
        type=MovieSceneScriptingFloatChannel
        default=0
        thinned_from=492
        keys [[-131,0.0024],[-130,0.001],[-129,0.0008],[-128,0.0003],[-127,0],[-126,0],[-125,0.0003],[-121,0],[-120,0.0004],[-119,0.0002],[-118,0.0007],[-117,0.0007],[-116,0.0011],[-114,0.0011],[-113,0.0008],[-111,0.0008],[-110,0.0004],[-109,0.0004],[-108,0.001],[-107,0.0013],[-104,0.0013],[-103,0.0008],[-102,0.0011],[-101,0.001],[-100,0.0007],[-98,0.0012],[-93,0.001],[-91,0.0008],[-90,0.0012],[-88,0.0014],[-84,0.0027],[-81,0.0027],[-77,0.0029],[-76,0.0032],[-75,0.0032],[-74,0.0029],[-73,0.0041],[-72,0.0043],[-71,0.0036],[-70,0.004],[-69,0.0046],[-68,0.0042],[-67,0.0052],[-66,0.005],[-65,0.005],[-64,0.0046],[-63,0.0044],[-62,0.0047],[-61,0.0046],[-59,0.0039],[-58,0.0034],[-49,0.0029],[-41,0.0025],[-39,0.003],[-38,0.0031],[-37,0.0034],[-36,0.0033],[-35,0.0035],[-34,0.0033],[-32,0.0034],[-31,0.0033],[-30,0.0027],[-28,0.0024],[-21,0.0025],[-19,0.0029],[-17,0.0028],[-16,0.0029],[-15,0.0028],[-10,0.0031],[-9,0.0029],[-2,0.0033],[-1,0.0032],[0,0.0035],[3,0.0032],[6,0.0034],[10,0.0033],[13,0.0042],[14,0.0041],[16,0.0034],[17,0.0036],[18,0.0041],[19,0.0043],[21,0.0042],[22,0.0037],[24,0.0038],[25,0.0035],[27,0.0039],[29,0.0034],[30,0.0036],[33,0.0031],[34,0.0033],[35,0.0032],[36,0.0029],[40,0.0034],[41,0.0032],[42,0.0036],[43,0.0037],[44,0.0034],[45,0.0033],[46,0.0037],[47,0.0032],[50,0.0027],[52,0.0028],[53,0.0027],[54,0.0029],[55,0.0028],[56,0.0024],[58,0.0008],[62,0.0015],[65,0.0013],[66,0.0019],[68,0.0015],[69,0.0018],[73,0.0012],[74,0.0013],[75,0.0018],[76,0.0015],[78,0.0016],[81,0.0012],[82,0.0009],[83,0.0012],[84,0.0008],[86,0.0009],[87,0.0013],[88,0.0009],[96,0.0007],[97,0.0005],[98,0.0007],[99,0.0014],[100,0.0017],[102,0.0014],[103,0.0007],[105,0.0009],[106,0.0007],[107,0.0008],[108,0.0006],[109,0.001],[111,0.0012],[114,0.0004],[115,0.0011],[116,0.001],[117,0.0005],[118,0.0008],[119,0.0006],[121,0.001],[122,0.0014],[124,0.0013],[125,0.0007],[126,0.002],[127,0.0025],[128,0.0018],[129,0.003],[130,0.0045],[131,0.0044],[132,0.0037],[133,0.0041],[134,0.0053],[135,0.0047],[136,0.0036],[137,0.0032],[138,0.0051],[139,0.0045],[140,0.0055],[143,0.0044],[145,0.0027],[146,0.0022],[147,0.0023],[149,0.0017],[150,0.002],[151,0.0028],[152,0.0029],[153,0.0032],[154,0.0029],[155,0.0031],[156,0.0027],[157,0.0031],[158,0.003],[159,0.0036],[160,0.003],[161,0.003],[162,0.0039],[164,0.003],[166,0.0025],[167,0.0025],[168,0.0016],[169,0.0023],[170,0.0024],[171,0.0028],[172,0.0024],[173,0.0015],[174,0.0009],[175,0.0012],[178,0.0015],[179,0.0007],[180,0.0006],[182,0.0006],[184,0.0002],[185,0.0004],[186,0.0004],[187,0.0007],[188,0.0008],[190,0.0001],[192,0.0003],[193,0.0016],[194,0.0025],[195,0.0025],[196,0.002],[197,0.0014],[198,0.0019],[199,0.0023],[201,0.0021],[202,0.0028],[203,0.0024],[205,0.0021],[206,0.0023],[208,0.0024],[210,0.0022],[212,0.0024],[214,0.0018],[216,0.0019],[217,0.0016],[219,0.0017],[220,0.002],[222,0.002],[223,0.0021],[224,0.002],[227,0.0022],[229,0.0016],[232,0.0017],[233,0.002],[234,0.0018],[235,0.0007],[236,0.002],[237,0.0024],[238,0.0024],[239,0.0011],[240,0.001],[241,0.0013],[242,0.0013],[244,0.0004],[245,0.0007],[246,0.0003],[247,0.0007],[248,0.0003],[249,0.0004],[250,0.0001],[251,0.0005],[252,0.0001],[253,0],[254,0.0005],[255,0.0005],[257,0],[259,0],[260,0.001],[261,0.0016],[262,0.0013],[263,0.0023],[265,0.0027],[268,0.0029],[269,0.0022],[272,0.0019],[274,0.0006],[277,0.0004],[279,0.0002],[281,0.0004],[282,0.0002],[283,0.0005],[285,0.0006],[286,0.0001],[288,0.0004],[291,0.0001],[292,0.0002],[293,0.0005],[294,0.0006],[296,0.0011],[297,0.0011],[298,0.0009],[299,0.0014],[300,0.0017],[306,0.002],[307,0.0022],[310,0.0019],[311,0.0017],[313,0.0015],[314,0.0016],[319,0.0015],[320,0.0018],[321,0.0017],[323,0.0011],[325,0.0018],[326,0.0017],[330,0.0026],[332,0.0026],[334,0.0023],[336,0.0026],[337,0.0024],[341,0.0021],[343,0.0022],[344,0.002],[345,0.0021],[350,0.0016],[351,0.002],[354,0.0015],[356,0.0018],[358,0.0017],[360,0.0015]]
       }
       CTRL_L_mouth_lipBiteD {
        type=MovieSceneScriptingFloatChannel
        default=0
        thinned_from=492
        keys [[-131,0.0018],[-129,0.0006],[-127,0],[-124,0.0002],[-123,0],[-121,0],[-120,0.0005],[-118,0.0002],[-117,0.0003],[-116,0.0006],[-114,0.0007],[-113,0.0005],[-112,0.0006],[-111,0.0008],[-109,0],[-107,0.0009],[-104,0.0009],[-103,0.0007],[-101,0.0009],[-100,0.0004],[-99,0.0002],[-97,0.0009],[-95,0.001],[-93,0.0008],[-92,0.0003],[-91,0.0002],[-89,0.0012],[-88,0.0013],[-86,0.0019],[-84,0.0021],[-83,0.0028],[-82,0.0042],[-81,0.0078],[-80,0.029],[-79,0.0344],[-78,0.0831],[-77,0.1584],[-76,0.2111],[-75,0.2873],[-74,0.3521],[-72,0.4023],[-71,0.4118],[-70,0.4198],[-69,0.4333],[-67,0.4314],[-66,0.4264],[-65,0.4167],[-64,0.4125],[-62,0.4164],[-61,0.406],[-60,0.385],[-59,0.3411],[-58,0.2601],[-57,0.2457],[-56,0.2093],[-55,0.1807],[-54,0.1628],[-53,0.1507],[-52,0.1453],[-51,0.114],[-50,0.0962],[-49,0.0945],[-48,0.0683],[-46,0.0692],[-45,0.0645],[-44,0.0626],[-43,0.0566],[-42,0.0572],[-41,0.0641],[-40,0.0777],[-39,0.1069],[-38,0.1182],[-37,0.2624],[-36,0.3313],[-35,0.3636],[-34,0.3657],[-33,0.3556],[-32,0.3031],[-31,0.2783],[-30,0.1508],[-29,0.1427],[-28,0.126],[-27,0.12],[-26,0.1182],[-25,0.1196],[-24,0.1206],[-23,0.1251],[-22,0.1103],[-21,0.0919],[-20,0.0894],[-19,0.0872],[-18,0.0772],[-17,0.0681],[-16,0.0575],[-15,0.0489],[-14,0.0448],[-13,0.0446],[-12,0.0424],[-11,0.0404],[-10,0.0412],[-9,0.0391],[-8,0.0352],[-7,0.0361],[-6,0.0383],[-5,0.0393],[-4,0.0384],[-3,0.0369],[-2,0.0381],[-1,0.0354],[0,0.0319],[1,0.0325],[2,0.0328],[3,0.0329],[4,0.0323],[5,0.0296],[6,0.0253],[8,0.0246],[10,0.0218],[11,0.0221],[12,0.0238],[13,0.0258],[15,0.0247],[16,0.0255],[19,0.0261],[20,0.0268],[21,0.0272],[23,0.027],[24,0.0271],[25,0.0286],[27,0.032],[28,0.0325],[29,0.032],[30,0.0312],[31,0.0308],[32,0.0312],[33,0.0321],[34,0.0306],[36,0.0307],[37,0.0332],[38,0.0326],[39,0.0318],[40,0.0325],[41,0.0313],[42,0.0305],[43,0.0331],[44,0.0353],[45,0.0356],[46,0.0321],[47,0.0337],[48,0.0412],[50,0.0467],[51,0.0469],[52,0.0482],[53,0.0524],[54,0.0494],[55,0.0412],[56,0.0364],[57,0.0343],[59,0.0249],[60,0.0242],[61,0.0244],[63,0.0239],[64,0.0188],[65,0.0188],[66,0.0198],[67,0.0222],[68,0.0222],[71,0.0229],[73,0.023],[74,0.0241],[75,0.0265],[76,0.0276],[77,0.0273],[78,0.0275],[79,0.0295],[80,0.0301],[81,0.0295],[82,0.0301],[83,0.0324],[84,0.0337],[85,0.0331],[86,0.0306],[87,0.0298],[88,0.0305],[91,0.0308],[92,0.0328],[93,0.0328],[94,0.0318],[95,0.0281],[96,0.0276],[97,0.028],[98,0.027],[99,0.0269],[100,0.0256],[101,0.0246],[103,0.0243],[104,0.0241],[105,0.024],[106,0.0226],[107,0.0228],[108,0.0229],[110,0.0209],[112,0.0178],[113,0.0158],[114,0.0146],[115,0.0142],[116,0.0158],[117,0.0155],[118,0.0161],[119,0.0181],[120,0.0196],[121,0.0193],[122,0.0175],[123,0.0173],[124,0.017],[125,0.0172],[126,0.0187],[127,0.0191],[129,0.0176],[130,0.0188],[131,0.0196],[132,0.0182],[133,0.0175],[134,0.0171],[136,0.0152],[138,0.015],[139,0.0141],[140,0.0179],[141,0.02],[142,0.0206],[143,0.019],[144,0.0169],[145,0.017],[146,0.0167],[147,0.0162],[148,0.0172],[149,0.0164],[150,0.0148],[152,0.0172],[153,0.019],[154,0.0194],[155,0.0182],[156,0.0177],[157,0.0163],[158,0.0159],[159,0.0182],[160,0.0187],[161,0.0175],[162,0.0166],[163,0.0154],[164,0.0146],[166,0.0158],[167,0.0149],[168,0.013],[170,0.0099],[171,0.0086],[173,0.0092],[174,0.0085],[175,0.0071],[176,0.0077],[178,0.0079],[179,0.006],[180,0.0048],[182,0.0047],[184,0.0033],[186,0.0022],[187,0.0011],[190,0.0005],[192,0.0008],[193,0.0017],[194,0.0037],[195,0.0037],[196,0.0028],[197,0.0016],[198,0.0018],[199,0.0034],[200,0.0039],[201,0.0053],[202,0.0108],[203,0.0116],[204,0.0072],[205,0.0071],[207,0.006],[208,0.0057],[209,0.0057],[211,0.0046],[212,0.0044],[214,0.0045],[215,0.0043],[216,0.0049],[218,0.004],[219,0.0041],[221,0.0056],[222,0.0061],[223,0.0071],[224,0.0075],[225,0.0065],[226,0.0063],[227,0.0068],[228,0.0063],[229,0.0062],[231,0.0076],[232,0.0076],[233,0.0086],[234,0.0082],[235,0.0043],[236,0.0025],[238,0.0026],[239,0.0015],[240,0.0012],[241,0.0015],[242,0.0016],[243,0.0008],[244,0.0004],[245,0.0006],[247,0.0002],[250,0.0001],[252,0.0002],[253,0],[255,0.0006],[257,0],[259,0],[261,0.0017],[262,0.0018],[263,0.0026],[265,0.0029],[268,0.0027],[269,0.0022],[270,0.0021],[271,0.0017],[272,0.0016],[273,0.0012],[274,0.0004],[276,0],[277,0],[278,0.0002],[280,0.0002],[281,0.0004],[284,0.0004],[285,0.0006],[286,0],[289,0.0005],[292,0],[295,0.0006],[296,0.001],[297,0.0011],[298,0.0009],[299,0.0012],[300,0.0019],[302,0.0018],[303,0.0023],[305,0.0061],[306,0.0074],[307,0.0083],[308,0.0085],[310,0.0083],[311,0.009],[312,0.0091],[313,0.0095],[314,0.0094],[315,0.0099],[316,0.0101],[319,0.009],[320,0.009],[321,0.0088],[322,0.0082],[323,0.0078],[327,0.01],[329,0.01],[331,0.0095],[332,0.009],[334,0.0086],[335,0.0088],[336,0.0093],[337,0.0092],[338,0.0085],[339,0.0085],[340,0.0088],[341,0.0081],[342,0.0079],[343,0.0081],[344,0.0078],[345,0.0082],[346,0.0082],[347,0.0079],[348,0.0083],[349,0.0081],[350,0.0072],[351,0.0067],[353,0.0071],[354,0.0068],[355,0.006],[356,0.0059],[358,0.0068],[359,0.0062],[360,0.0062]]
       }
       CTRL_R_mouth_purseU {
        type=MovieSceneScriptingFloatChannel
        default=0
        thinned_from=492
        keys [[-131,0.1344],[-130,0.1317],[-129,0.1237],[-128,0.1166],[-127,0.1161],[-126,0.1165],[-125,0.122],[-123,0.1216],[-122,0.1218],[-121,0.1224],[-120,0.1271],[-119,0.1288],[-118,0.1256],[-117,0.1242],[-116,0.1151],[-115,0.1145],[-114,0.1157],[-113,0.1157],[-112,0.1141],[-111,0.1119],[-110,0.1104],[-109,0.1094],[-108,0.1035],[-107,0.1],[-106,0.0995],[-105,0.097],[-104,0.0996],[-103,0.0973],[-102,0.099],[-101,0.103],[-100,0.0989],[-99,0.1018],[-98,0.1026],[-97,0.1024],[-96,0.1016],[-95,0.1016],[-94,0.1113],[-93,0.1115],[-92,0.108],[-91,0.1122],[-89,0.1189],[-88,0.1274],[-87,0.1426],[-86,0.1523],[-85,0.1627],[-84,0.1766],[-83,0.2],[-82,0.2506],[-81,0.2995],[-80,0.3444],[-79,0.3513],[-78,0.4111],[-77,0.4389],[-76,0.4657],[-75,0.497],[-74,0.5012],[-72,0.5269],[-71,0.5207],[-69,0.5266],[-68,0.5182],[-67,0.5168],[-66,0.5234],[-65,0.522],[-64,0.5182],[-62,0.506],[-61,0.4914],[-60,0.451],[-59,0.3742],[-58,0.2819],[-57,0.2576],[-56,0.2055],[-55,0.1784],[-54,0.1609],[-53,0.1499],[-52,0.1445],[-51,0.1371],[-50,0.1328],[-49,0.1323],[-48,0.1322],[-47,0.1384],[-46,0.1425],[-45,0.1604],[-44,0.1681],[-43,0.1977],[-42,0.2346],[-41,0.2677],[-40,0.3265],[-39,0.39],[-38,0.4068],[-37,0.4652],[-35,0.4854],[-34,0.4751],[-33,0.4462],[-32,0.3588],[-31,0.333],[-30,0.2132],[-29,0.1999],[-28,0.1698],[-27,0.1516],[-26,0.1426],[-25,0.1353],[-24,0.1332],[-23,0.1384],[-22,0.1381],[-21,0.1407],[-20,0.1383],[-19,0.138],[-18,0.1392],[-16,0.1207],[-15,0.117],[-14,0.1216],[-13,0.1192],[-12,0.1146],[-10,0.1124],[-9,0.1108],[-8,0.1088],[-7,0.1026],[-4,0.0989],[-3,0.0996],[-2,0.0966],[-1,0.0965],[0,0.0946],[1,0.0961],[2,0.0962],[3,0.098],[4,0.0985],[5,0.1006],[6,0.1052],[7,0.1091],[8,0.1093],[10,0.1113],[11,0.1086],[12,0.1086],[13,0.1097],[14,0.113],[15,0.1144],[16,0.1139],[17,0.1163],[18,0.1145],[19,0.1161],[20,0.1198],[21,0.1171],[22,0.1146],[23,0.1148],[24,0.1158],[27,0.1177],[28,0.1172],[29,0.1196],[31,0.1158],[32,0.1189],[33,0.1192],[34,0.1173],[35,0.1177],[36,0.1199],[38,0.1156],[39,0.1157],[40,0.1184],[41,0.1204],[42,0.1198],[43,0.1179],[44,0.1174],[45,0.1178],[46,0.1162],[47,0.1174],[48,0.1238],[49,0.1261],[50,0.1226],[51,0.1286],[52,0.1321],[53,0.1317],[54,0.1297],[55,0.1257],[56,0.125],[57,0.1249],[58,0.123],[59,0.1197],[60,0.119],[61,0.1204],[62,0.1189],[63,0.1188],[64,0.116],[65,0.116],[66,0.1155],[67,0.1161],[68,0.1174],[69,0.1181],[70,0.1173],[71,0.1162],[72,0.1156],[73,0.1136],[74,0.1172],[75,0.1216],[76,0.1198],[77,0.1143],[78,0.1141],[79,0.1117],[80,0.1101],[81,0.1109],[82,0.1076],[83,0.1067],[84,0.1088],[85,0.1083],[86,0.1017],[87,0.1009],[88,0.0972],[89,0.0968],[90,0.0957],[91,0.0993],[92,0.1008],[93,0.1004],[95,0.098],[96,0.0975],[97,0.0953],[99,0.0948],[101,0.0954],[102,0.0949],[103,0.0919],[104,0.0923],[105,0.0923],[106,0.0936],[107,0.0934],[108,0.0914],[109,0.0947],[110,0.0976],[111,0.1003],[113,0.1007],[114,0.0997],[115,0.1039],[116,0.1051],[117,0.1097],[118,0.1205],[119,0.1207],[120,0.1224],[121,0.1189],[122,0.1176],[123,0.1229],[124,0.1225],[125,0.1185],[126,0.1216],[127,0.1188],[128,0.1204],[129,0.1154],[130,0.112],[131,0.1113],[132,0.111],[133,0.1068],[134,0.1095],[135,0.1162],[136,0.1224],[137,0.119],[138,0.1048],[139,0.0946],[140,0.0904],[141,0.0904],[143,0.0983],[144,0.101],[145,0.1001],[146,0.1027],[147,0.1063],[149,0.1012],[150,0.1004],[151,0.1],[152,0.0989],[153,0.0988],[154,0.1],[155,0.1023],[156,0.0994],[157,0.0953],[158,0.0935],[159,0.0974],[160,0.096],[161,0.0965],[162,0.0912],[163,0.092],[164,0.0938],[165,0.0926],[166,0.0893],[167,0.095],[168,0.1028],[169,0.1068],[170,0.108],[171,0.1064],[172,0.1109],[173,0.1163],[174,0.1221],[175,0.1233],[176,0.1235],[177,0.1225],[179,0.1139],[180,0.1132],[181,0.1208],[182,0.1251],[183,0.125],[184,0.1164],[185,0.114],[186,0.1195],[187,0.1254],[188,0.1262],[190,0.1272],[191,0.1261],[192,0.134],[193,0.1312],[194,0.1243],[195,0.1241],[196,0.1208],[197,0.1214],[198,0.1278],[199,0.1362],[200,0.1387],[201,0.1399],[202,0.1255],[203,0.1166],[204,0.1118],[205,0.1134],[206,0.1123],[207,0.1101],[209,0.1099],[210,0.1076],[211,0.109],[212,0.1089],[213,0.1116],[214,0.1133],[215,0.112],[216,0.1135],[218,0.1178],[220,0.1184],[221,0.1183],[223,0.1213],[224,0.122],[225,0.1209],[226,0.122],[228,0.1258],[229,0.1207],[231,0.12],[233,0.1217],[234,0.1206],[235,0.1292],[236,0.1422],[237,0.1566],[238,0.1519],[239,0.1488],[240,0.1536],[241,0.1669],[242,0.1696],[243,0.1682],[244,0.1803],[245,0.1828],[246,0.2078],[247,0.2261],[248,0.2253],[249,0.2008],[250,0.1798],[251,0.1805],[252,0.1635],[253,0.1568],[254,0.1655],[255,0.1637],[256,0.1623],[257,0.1692],[258,0.1949],[259,0.1932],[260,0.1756],[261,0.1588],[262,0.1366],[263,0.1142],[264,0.1113],[265,0.1104],[267,0.1072],[268,0.1107],[269,0.1057],[270,0.1044],[271,0.1068],[272,0.115],[273,0.1239],[274,0.1206],[275,0.1267],[276,0.1256],[277,0.1223],[278,0.1221],[279,0.1225],[280,0.1239],[281,0.1216],[282,0.1235],[283,0.1231],[284,0.1231],[285,0.1245],[286,0.116],[288,0.1184],[289,0.118],[290,0.1186],[291,0.1159],[292,0.1118],[293,0.1126],[294,0.1087],[295,0.109],[296,0.1244],[297,0.1274],[298,0.1376],[299,0.1397],[300,0.1423],[301,0.1395],[302,0.1165],[303,0.1145],[304,0.1088],[305,0.1073],[306,0.1061],[308,0.1028],[309,0.1042],[310,0.1051],[311,0.1076],[312,0.1084],[313,0.1104],[314,0.1109],[315,0.1085],[316,0.1045],[317,0.104],[318,0.0966],[319,0.0995],[320,0.103],[321,0.102],[322,0.1026],[323,0.1051],[324,0.1064],[325,0.1044],[326,0.1092],[327,0.1124],[328,0.1151],[329,0.1141],[330,0.116],[331,0.1151],[332,0.1127],[334,0.1102],[335,0.1124],[336,0.1122],[337,0.1099],[338,0.112],[339,0.1117],[340,0.1136],[341,0.1079],[342,0.1092],[343,0.1083],[344,0.1064],[345,0.1093],[346,0.1125],[348,0.1081],[349,0.1069],[351,0.1028],[352,0.1017],[353,0.1051],[354,0.1027],[355,0.1028],[356,0.1035],[357,0.1061],[358,0.1047],[359,0.1011],[360,0.1011]]
       }
       CTRL_R_mouth_cornerDepress {
        type=MovieSceneScriptingFloatChannel
        default=0
        thinned_from=492
        keys [[-131,0.0474],[-130,0.0551],[-129,0.0572],[-127,0.0546],[-126,0.054],[-125,0.0507],[-124,0.0455],[-123,0.0445],[-122,0.0444],[-121,0.0436],[-120,0.0418],[-119,0.041],[-118,0.0406],[-117,0.0406],[-116,0.0401],[-115,0.0399],[-114,0.0416],[-113,0.0436],[-112,0.0459],[-111,0.0479],[-110,0.0485],[-109,0.0493],[-108,0.0543],[-107,0.0603],[-106,0.0612],[-105,0.0627],[-104,0.0651],[-103,0.0661],[-102,0.0697],[-101,0.0724],[-100,0.0758],[-98,0.084],[-97,0.0896],[-96,0.0928],[-95,0.0936],[-94,0.0991],[-93,0.0991],[-92,0.096],[-91,0.09],[-90,0.0816],[-89,0.0682],[-88,0.0529],[-87,0.0386],[-86,0.0294],[-85,0.0225],[-84,0.016],[-83,0.0124],[-82,0.0104],[-81,0.0095],[-79,0.0097],[-78,0.0115],[-77,0.0192],[-76,0.0272],[-75,0.0398],[-74,0.0521],[-73,0.0609],[-72,0.0669],[-71,0.0689],[-70,0.0697],[-69,0.0711],[-68,0.0732],[-67,0.0705],[-65,0.0669],[-64,0.0672],[-63,0.0628],[-62,0.0581],[-61,0.0537],[-60,0.0511],[-59,0.0494],[-58,0.045],[-57,0.043],[-56,0.0385],[-54,0.0311],[-53,0.0287],[-52,0.0276],[-51,0.0256],[-50,0.0242],[-49,0.0241],[-48,0.0171],[-47,0.0144],[-46,0.0145],[-44,0.0138],[-43,0.0128],[-42,0.0125],[-41,0.0131],[-40,0.0126],[-39,0.0115],[-38,0.0113],[-37,0.0193],[-36,0.0261],[-35,0.0325],[-34,0.0358],[-33,0.0369],[-32,0.0339],[-31,0.0325],[-30,0.0243],[-29,0.0233],[-27,0.0187],[-26,0.0162],[-25,0.0139],[-24,0.0133],[-23,0.0136],[-22,0.0134],[-21,0.0139],[-20,0.0139],[-19,0.0196],[-18,0.0246],[-17,0.0284],[-16,0.0271],[-15,0.0252],[-14,0.0226],[-13,0.0213],[-11,0.0199],[-10,0.0207],[-9,0.0228],[-8,0.0244],[-7,0.025],[-6,0.0245],[-5,0.0227],[-3,0.0202],[-2,0.0205],[-1,0.0204],[0,0.0212],[3,0.0217],[4,0.0224],[5,0.024],[6,0.0228],[7,0.0209],[8,0.0203],[9,0.0191],[10,0.0191],[11,0.0201],[12,0.022],[13,0.0233],[14,0.0238],[16,0.0214],[17,0.0216],[18,0.0229],[19,0.0236],[20,0.0233],[21,0.0217],[22,0.0204],[23,0.0202],[25,0.0189],[26,0.0188],[27,0.0193],[28,0.0194],[30,0.0187],[31,0.019],[33,0.0187],[34,0.0188],[35,0.0187],[38,0.0176],[39,0.0182],[40,0.0191],[41,0.0198],[42,0.0199],[45,0.0194],[46,0.0168],[47,0.0147],[48,0.0145],[49,0.0168],[50,0.02],[51,0.0246],[52,0.0273],[53,0.0274],[54,0.0253],[55,0.0208],[56,0.0179],[57,0.0162],[59,0.0172],[60,0.0173],[61,0.0189],[62,0.018],[63,0.0179],[64,0.0167],[67,0.0167],[68,0.0169],[69,0.0168],[70,0.0155],[72,0.0112],[73,0.0103],[74,0.0098],[75,0.0091],[76,0.0072],[77,0.0055],[78,0.0055],[79,0.0048],[80,0.005],[81,0.0045],[82,0.0031],[83,0.0028],[85,0.0029],[86,0.0025],[87,0.0027],[89,0.0027],[90,0.002],[91,0.0021],[92,0.0025],[94,0.0027],[95,0.003],[96,0.0036],[98,0.004],[100,0.0049],[102,0.0046],[103,0.0054],[104,0.006],[105,0.006],[107,0.0055],[108,0.0051],[110,0.0055],[111,0.0063],[112,0.0064],[113,0.0062],[114,0.0056],[115,0.0056],[118,0.0075],[119,0.0077],[120,0.0081],[122,0.0063],[123,0.005],[124,0.0049],[126,0.0038],[127,0.0039],[128,0.0044],[129,0.0058],[131,0.008],[132,0.0076],[133,0.0075],[134,0.0097],[135,0.0138],[136,0.0153],[137,0.0175],[139,0.0124],[140,0.0081],[142,0.0056],[144,0.0026],[145,0.0031],[147,0.003],[148,0.0017],[150,0.0028],[151,0.0038],[153,0.0038],[154,0.0033],[156,0.0029],[157,0.0035],[158,0.0038],[159,0.0046],[160,0.0046],[161,0.004],[164,0.0032],[165,0.0024],[167,0.002],[168,0.0019],[171,0.0038],[172,0.0042],[173,0.0043],[174,0.005],[175,0.007],[176,0.0091],[177,0.0096],[178,0.0109],[180,0.0125],[181,0.0148],[182,0.0166],[183,0.0172],[185,0.017],[186,0.0183],[187,0.019],[188,0.0191],[189,0.0204],[191,0.0224],[192,0.0261],[193,0.0308],[194,0.036],[195,0.0358],[196,0.0411],[197,0.0484],[198,0.0559],[199,0.0564],[200,0.0554],[201,0.0483],[202,0.0594],[203,0.0625],[204,0.0679],[205,0.0513],[206,0.0511],[208,0.0526],[209,0.053],[210,0.0497],[211,0.0462],[212,0.0453],[213,0.0456],[214,0.0469],[215,0.0464],[216,0.0454],[217,0.0458],[218,0.0447],[219,0.044],[220,0.042],[221,0.0409],[222,0.0405],[223,0.0408],[224,0.0427],[225,0.0433],[226,0.0427],[227,0.0414],[228,0.0407],[229,0.0415],[230,0.0401],[232,0.0359],[233,0.0369],[234,0.0389],[235,0.0386],[236,0.0353],[237,0.0329],[238,0.0326],[239,0.0331],[240,0.0338],[241,0.0349],[242,0.0351],[243,0.0377],[244,0.0379],[245,0.0391],[246,0.0453],[247,0.0533],[248,0.0585],[249,0.0575],[250,0.0527],[251,0.0481],[252,0.0403],[253,0.0316],[254,0.0236],[255,0.0186],[256,0.0187],[257,0.0177],[258,0.0214],[259,0.0221],[260,0.0277],[261,0.0286],[262,0.0279],[263,0.0264],[264,0.0265],[265,0.0272],[267,0.0308],[268,0.0335],[269,0.0379],[270,0.0386],[271,0.0432],[272,0.0439],[273,0.0434],[274,0.0413],[275,0.0401],[276,0.0385],[277,0.0374],[278,0.0367],[279,0.0374],[280,0.0385],[281,0.0376],[282,0.0369],[284,0.0371],[285,0.0387],[286,0.0389],[287,0.0404],[288,0.0428],[289,0.0477],[291,0.0514],[292,0.0523],[294,0.0519],[295,0.0532],[296,0.0556],[297,0.0558],[298,0.0532],[299,0.0481],[300,0.0379],[301,0.0368],[302,0.0278],[303,0.0263],[304,0.0217],[305,0.0193],[306,0.0174],[307,0.0167],[309,0.0145],[310,0.0142],[311,0.0125],[312,0.0123],[313,0.0114],[314,0.0102],[316,0.0102],[319,0.0092],[320,0.0097],[321,0.0099],[323,0.0095],[324,0.009],[325,0.0091],[328,0.0086],[329,0.0087],[330,0.0091],[331,0.0093],[333,0.0087],[334,0.0081],[337,0.0073],[339,0.0075],[340,0.008],[342,0.0077],[343,0.0079],[344,0.0084],[345,0.0085],[347,0.0083],[349,0.0085],[351,0.008],[352,0.0079],[353,0.0081],[356,0.0078],[357,0.0081],[358,0.0081],[359,0.0083],[360,0.0083]]
       }
       CTRL_R_mouth_upperLipRaise {
        type=MovieSceneScriptingFloatChannel
        default=0
        thinned_from=492
        keys [[-131,0.0019],[-130,0.0007],[-129,0.0005],[-127,0],[-119,0],[-116,0.0004],[-114,0.0003],[-113,0.0001],[-111,0.0002],[-110,0],[-109,0],[-107,0.0004],[-104,0.0006],[-103,0.0002],[-102,0.0002],[-101,0],[-99,0],[-95,0.0004],[-93,0.0003],[-92,0],[-91,0.0001],[-89,0.0008],[-88,0.001],[-83,0.0023],[-81,0.0022],[-79,0.0023],[-77,0.002],[-76,0.0021],[-74,0.0016],[-73,0.0022],[-71,0.0022],[-70,0.0024],[-69,0.0021],[-67,0.0023],[-65,0.0018],[-61,0.0018],[-59,0.0023],[-54,0.0021],[-51,0.0024],[-46,0.0022],[-43,0.0024],[-41,0.0021],[-38,0.0024],[-34,0.0018],[-32,0.0023],[-28,0.0019],[-27,0.0021],[-26,0.0019],[-24,0.0021],[-23,0.0017],[-22,0.0019],[-21,0.0016],[-20,0.0022],[-18,0.0026],[-15,0.0023],[-6,0.0028],[-4,0.0027],[-3,0.0029],[2,0.0027],[4,0.0025],[5,0.0026],[12,0.0022],[18,0.0022],[19,0.0024],[20,0.0023],[21,0.0024],[22,0.0021],[29,0.0024],[30,0.0025],[32,0.0024],[33,0.0021],[42,0.0021],[47,0.0022],[48,0.0023],[52,0.0024],[53,0.0021],[54,0.0022],[56,0.0018],[58,0.0005],[60,0.0009],[61,0.0008],[62,0.0012],[65,0.0013],[66,0.0018],[67,0.0011],[68,0.001],[69,0.0015],[74,0.001],[75,0.0015],[76,0.0012],[78,0.0013],[79,0.0009],[80,0.0008],[81,0.0009],[82,0.0006],[83,0.0008],[84,0.0006],[86,0.0009],[87,0.0013],[88,0.0006],[94,0.0001],[98,0.0002],[100,0.0011],[102,0.0007],[103,0.0002],[105,0.0004],[106,0.0002],[107,0.0004],[108,0.0003],[111,0.0006],[114,0.0001],[115,0.0005],[116,0.0005],[117,0.0002],[118,0.0002],[119,0],[121,0.0004],[122,0.0003],[123,0],[132,0],[133,0.0003],[134,0.0009],[136,0.0012],[137,0.0017],[138,0.0037],[139,0.0015],[140,0.0032],[141,0.0028],[142,0.003],[143,0.0028],[144,0.0018],[146,0.0007],[147,0.001],[149,0.0006],[150,0.0009],[151,0.0014],[152,0.0011],[153,0.0014],[154,0.0013],[155,0.0014],[156,0.0008],[157,0.0008],[159,0.0015],[160,0.001],[162,0.0017],[163,0.0014],[164,0.0008],[165,0.001],[166,0.001],[167,0.0013],[168,0.0013],[169,0.0022],[171,0.0024],[172,0.0021],[173,0.0012],[174,0.0006],[175,0.0007],[176,0.001],[178,0.0012],[179,0.0005],[181,0.0007],[182,0.0006],[184,0],[185,0.0005],[189,0.0003],[190,0.0001],[191,0.0003],[192,0.0003],[194,0.002],[195,0.002],[197,0.0012],[198,0.0019],[199,0.0023],[200,0.0023],[201,0.0018],[202,0.0028],[203,0.0023],[205,0.002],[208,0.0021],[210,0.002],[211,0.0021],[212,0.0021],[214,0.0015],[215,0.0018],[216,0.0017],[217,0.0014],[219,0.0015],[220,0.0019],[222,0.0016],[223,0.0017],[224,0.0015],[227,0.0019],[229,0.0016],[231,0.0017],[232,0.0016],[233,0.0018],[234,0.0016],[235,0.0005],[236,0.0019],[237,0.0021],[238,0.002],[239,0.001],[240,0.001],[241,0.0013],[242,0.0014],[243,0.0008],[245,0.0009],[246,0.0004],[247,0.0005],[248,0.0002],[249,0.0002],[250,0],[251,0.0003],[252,0],[255,0.0004],[257,0],[259,0],[260,0.001],[261,0.0016],[262,0.0013],[263,0.0021],[265,0.0024],[268,0.002],[269,0.0016],[270,0.0015],[271,0.0016],[272,0.0012],[274,0.0002],[294,0],[295,0.0004],[297,0.0007],[298,0.0009],[300,0.0018],[301,0.0018],[302,0.0016],[307,0.0022],[309,0.0018],[313,0.0013],[316,0.0015],[317,0.0012],[318,0.0011],[319,0.0015],[320,0.0016],[321,0.0015],[322,0.001],[323,0.001],[324,0.0015],[326,0.0017],[327,0.0014],[329,0.0017],[330,0.0016],[332,0.0019],[333,0.0016],[334,0.0016],[335,0.0018],[337,0.0016],[338,0.0017],[340,0.0013],[343,0.0016],[348,0.0015],[350,0.001],[351,0.0013],[353,0.0015],[358,0.0015],[360,0.0013]]
       }
       CTRL_R_mouth_dimple {
        type=MovieSceneScriptingFloatChannel
        default=0.7036
        thinned_from=492
        keys [[-131,0.0016],[-129,0.0013],[-127,0.0004],[-125,0.0001],[-116,0.0005],[-114,0.0008],[-108,0.0004],[-104,0.0009],[-102,0.0008],[-101,0.001],[-98,0.0005],[-93,0.0008],[-91,0.0005],[-90,0.0005],[-88,0.0011],[-86,0.002],[-85,0.0029],[-84,0.0055],[-83,0.0115],[-82,0.0222],[-81,0.0331],[-80,0.0472],[-79,0.0485],[-78,0.0597],[-77,0.072],[-76,0.0837],[-75,0.0929],[-73,0.1147],[-72,0.1187],[-71,0.1274],[-70,0.1281],[-69,0.1262],[-68,0.127],[-67,0.1232],[-66,0.1232],[-65,0.1247],[-64,0.1281],[-63,0.1302],[-62,0.1332],[-61,0.1345],[-60,0.1308],[-59,0.1213],[-58,0.0978],[-57,0.0914],[-56,0.0759],[-55,0.064],[-54,0.0555],[-53,0.0512],[-52,0.0496],[-51,0.0414],[-50,0.0363],[-49,0.0358],[-48,0.0276],[-46,0.0299],[-45,0.0379],[-44,0.0401],[-43,0.0475],[-42,0.0538],[-41,0.0595],[-40,0.0708],[-39,0.0841],[-38,0.0879],[-37,0.1179],[-36,0.134],[-35,0.151],[-34,0.1584],[-33,0.1574],[-32,0.1458],[-31,0.1402],[-30,0.1034],[-29,0.0957],[-28,0.0769],[-27,0.0649],[-26,0.061],[-25,0.0605],[-24,0.0604],[-22,0.0476],[-20,0.03],[-19,0.0238],[-18,0.0178],[-17,0.0138],[-16,0.0105],[-15,0.01],[-14,0.009],[-13,0.0095],[-12,0.009],[-11,0.0089],[-10,0.0095],[-9,0.0081],[-8,0.008],[-7,0.0085],[-6,0.0078],[-5,0.0094],[-4,0.0099],[-3,0.0095],[-2,0.0098],[-1,0.008],[0,0.0074],[2,0.0052],[3,0.005],[4,0.0044],[5,0.0043],[6,0.0035],[12,0.0025],[26,0.0025],[30,0.0027],[31,0.0025],[42,0.0025],[45,0.0029],[47,0.0037],[50,0.0034],[51,0.0029],[52,0.003],[53,0.0089],[54,0.0124],[55,0.0211],[56,0.0281],[57,0.0339],[58,0.0349],[59,0.0337],[60,0.0334],[61,0.0338],[62,0.0353],[63,0.0355],[64,0.0386],[65,0.0386],[66,0.0395],[67,0.0387],[68,0.0384],[70,0.0346],[72,0.035],[73,0.038],[74,0.0415],[75,0.0454],[76,0.0514],[77,0.0552],[78,0.0554],[79,0.0579],[80,0.0593],[82,0.0598],[83,0.0588],[84,0.0588],[85,0.0575],[86,0.0582],[88,0.0582],[89,0.0586],[91,0.0583],[92,0.0554],[93,0.0545],[94,0.052],[95,0.0505],[96,0.0504],[97,0.0498],[98,0.049],[99,0.0478],[100,0.0474],[102,0.0471],[103,0.0462],[104,0.0466],[105,0.0466],[106,0.0476],[107,0.0483],[108,0.05],[109,0.0511],[110,0.0534],[111,0.0577],[112,0.0601],[113,0.0606],[114,0.0606],[115,0.0615],[116,0.0633],[117,0.0671],[118,0.0713],[119,0.0738],[121,0.076],[122,0.0776],[123,0.0773],[124,0.0777],[125,0.0766],[126,0.0742],[127,0.0731],[129,0.0695],[130,0.0657],[131,0.0636],[132,0.0619],[133,0.0611],[134,0.0607],[135,0.0613],[136,0.0646],[138,0.066],[139,0.0718],[140,0.0771],[141,0.0837],[142,0.104],[143,0.1084],[144,0.1102],[145,0.1053],[146,0.0995],[147,0.0983],[148,0.1007],[149,0.1024],[150,0.1034],[151,0.1029],[152,0.101],[153,0.1025],[154,0.1044],[155,0.1027],[157,0.1005],[158,0.0971],[159,0.0972],[160,0.0947],[161,0.0959],[162,0.0984],[163,0.0993],[164,0.1014],[165,0.1015],[167,0.1057],[168,0.1048],[169,0.1005],[170,0.0947],[171,0.0903],[172,0.0847],[173,0.0768],[174,0.0713],[175,0.0641],[176,0.0604],[177,0.0596],[178,0.0568],[179,0.0546],[180,0.0531],[181,0.0482],[183,0.0397],[184,0.0369],[185,0.0347],[186,0.0318],[187,0.0281],[188,0.0275],[189,0.0244],[190,0.0202],[192,0.0201],[193,0.0207],[194,0.0253],[195,0.0264],[196,0.0265],[198,0.0169],[199,0.0155],[200,0.0153],[201,0.026],[202,0.0258],[203,0.0284],[204,0.0274],[205,0.0192],[206,0.0226],[207,0.0216],[208,0.0196],[209,0.0183],[210,0.0182],[211,0.019],[212,0.0188],[213,0.0187],[215,0.0174],[216,0.0176],[217,0.0168],[218,0.0169],[219,0.0173],[220,0.017],[221,0.0195],[223,0.0187],[224,0.0174],[225,0.0152],[226,0.0153],[227,0.0153],[228,0.0162],[229,0.0163],[230,0.0178],[231,0.0186],[232,0.0184],[233,0.0191],[234,0.0172],[235,0.0164],[236,0.0164],[238,0.012],[239,0.0073],[240,0.0043],[241,0.0022],[242,0.0018],[243,0.001],[244,0.0008],[245,0.0008],[246,0.0006],[253,0],[255,0.0003],[259,0],[262,0.0013],[263,0.0021],[264,0.0022],[266,0.0028],[268,0.0027],[269,0.0024],[271,0.0022],[273,0.0015],[276,0],[279,0],[284,0.0004],[294,0.0002],[296,0.0006],[298,0.0009],[301,0.0018],[302,0.0042],[303,0.0048],[304,0.0071],[306,0.0125],[307,0.0157],[309,0.0193],[310,0.0199],[311,0.0194],[312,0.0194],[313,0.02],[314,0.0198],[317,0.0219],[318,0.0233],[319,0.0238],[320,0.0251],[322,0.0255],[323,0.0252],[324,0.0253],[326,0.0261],[327,0.027],[328,0.027],[329,0.0273],[331,0.0284],[332,0.0295],[334,0.0301],[335,0.0313],[336,0.0318],[337,0.033],[338,0.0338],[339,0.0334],[340,0.0336],[341,0.0333],[343,0.0337],[345,0.0333],[347,0.0323],[349,0.0319],[350,0.032],[351,0.0319],[352,0.0307],[353,0.0302],[354,0.0292],[355,0.0287],[356,0.0287],[357,0.0277],[358,0.0261],[359,0.0243],[360,0.0243]]
       }
       CTRL_R_mouth_cornerPull {
        type=MovieSceneScriptingFloatChannel
        default=0
        thinned_from=492
        keys [[-131,0.0015],[-129,0.0009],[-128,0.0003],[-126,0],[-121,0],[-120,0.0002],[-116,0.0004],[-114,0.0007],[-110,0.0004],[-109,0.0002],[-107,0.0007],[-105,0.001],[-103,0.0007],[-99,0.0004],[-94,0.0009],[-93,0.0009],[-91,0.0002],[-85,0.0021],[-75,0.0024],[-72,0.0021],[-70,0.0023],[-61,0.0022],[-57,0.0024],[-52,0.0022],[-38,0.002],[-37,0.0023],[-30,0.0024],[-27,0.0019],[-21,0.0014],[-18,0.0024],[-16,0.0024],[-15,0.0023],[-11,0.0025],[-4,0.0021],[5,0.0022],[6,0.0024],[21,0.0021],[24,0.002],[32,0.002],[42,0.0019],[45,0.0022],[47,0.002],[50,0.0023],[56,0.0017],[59,0.0003],[60,0.0002],[62,0.0007],[74,0.0005],[75,0.0008],[78,0.0008],[79,0.0004],[81,0.0021],[82,0.0033],[83,0.0036],[84,0.0026],[85,0.0021],[86,0.0027],[87,0.0038],[88,0.0043],[89,0.004],[90,0.0047],[91,0.0043],[92,0.0026],[94,0.0027],[96,0.0036],[97,0.0035],[98,0.0026],[99,0.0025],[101,0.0041],[102,0.0043],[103,0.003],[105,0.0032],[106,0.0038],[108,0.0039],[110,0.0031],[111,0.0021],[112,0.002],[113,0.0013],[114,0.0011],[115,0.0007],[116,0.0011],[119,0.0001],[132,0],[135,0.0005],[136,0.0004],[137,0.0007],[138,0.0041],[139,0.0117],[140,0.0218],[141,0.0264],[142,0.035],[143,0.0347],[144,0.0327],[145,0.027],[146,0.0261],[147,0.0259],[148,0.0279],[149,0.0266],[150,0.0262],[151,0.0268],[153,0.0286],[154,0.0283],[155,0.0268],[156,0.0261],[157,0.0262],[158,0.0278],[159,0.0274],[160,0.0262],[161,0.028],[162,0.0302],[163,0.031],[164,0.0303],[165,0.0303],[166,0.0316],[167,0.0314],[168,0.0277],[169,0.0213],[171,0.0143],[172,0.0102],[173,0.0066],[174,0.0032],[175,0.001],[176,0.0001],[177,0.0002],[178,0.0005],[180,0.0004],[181,0],[192,0],[193,0.0005],[194,0.0014],[196,0.0016],[197,0.0011],[198,0.0011],[204,0.0019],[205,0.0016],[213,0.0016],[215,0.0013],[218,0.0013],[219,0.0011],[228,0.0016],[234,0.0013],[235,0.0009],[236,0.001],[238,0.0015],[240,0.0006],[243,0.0008],[246,0.0003],[247,0.0004],[259,0],[261,0.001],[264,0.0019],[265,0.0023],[266,0.0025],[268,0.0022],[269,0.0019],[272,0.0015],[273,0.0012],[274,0.0005],[275,0.0002],[278,0],[285,0.0002],[287,0],[290,0.0001],[294,0],[298,0.0006],[299,0.0009],[300,0.0017],[305,0.0017],[307,0.002],[310,0.0019],[311,0.0015],[317,0.0014],[319,0.0011],[321,0.0015],[323,0.0011],[324,0.0011],[326,0.0015],[329,0.0015],[332,0.0017],[335,0.0014],[338,0.0014],[341,0.0011],[345,0.0014],[350,0.001],[351,0.0012],[352,0.0012],[353,0.0014],[357,0.0012],[360,0.0012]]
       }
      ]
     }
     {
      type=MovieSceneControlRigParameterSection
      channels [
       CTRL_C_eyesAim.Location.X {type=MovieSceneScriptingFloatChannel, default=0, const [0,6]}
       CTRL_C_eyesAim.Location.Y {type=MovieSceneScriptingFloatChannel, default=0, const [0,6]}
       CTRL_C_eyesAim.Location.Z {type=MovieSceneScriptingFloatChannel, default=0, const [0,6]}
       CTRL_C_eyesAim.Rotation.X {type=MovieSceneScriptingFloatChannel, default=0, const [0,6]}
       CTRL_C_eyesAim.Rotation.Y {type=MovieSceneScriptingFloatChannel, default=0, const [0,6]}
       CTRL_C_eyesAim.Rotation.Z {type=MovieSceneScriptingFloatChannel, default=0, const [0,6]}
       CTRL_C_eyesAim.Scale.X {type=MovieSceneScriptingFloatChannel, default=0, const [0,6]}
       CTRL_C_eyesAim.Scale.Y {type=MovieSceneScriptingFloatChannel, default=0, const [0,6]}
       CTRL_C_eyesAim.Scale.Z {type=MovieSceneScriptingFloatChannel, default=0, const [0,6]}
       CTRL_L_eyeAim.Location.X {type=MovieSceneScriptingFloatChannel, default=0, const [0,6]}
       CTRL_L_eyeAim.Location.Y {type=MovieSceneScriptingFloatChannel, default=0, const [0,6]}
       CTRL_L_eyeAim.Location.Z {type=MovieSceneScriptingFloatChannel, default=0, const [0,6]}
       CTRL_L_eyeAim.Rotation.X {type=MovieSceneScriptingFloatChannel, default=0, const [0,6]}
       CTRL_L_eyeAim.Rotation.Y {type=MovieSceneScriptingFloatChannel, default=0, const [0,6]}
       CTRL_L_eyeAim.Rotation.Z {type=MovieSceneScriptingFloatChannel, default=0, const [0,6]}
       CTRL_L_eyeAim.Scale.X {type=MovieSceneScriptingFloatChannel, default=0, const [0,6]}
       CTRL_L_eyeAim.Scale.Y {type=MovieSceneScriptingFloatChannel, default=0, const [0,6]}
       CTRL_L_eyeAim.Scale.Z {type=MovieSceneScriptingFloatChannel, default=0, const [0,6]}
       CTRL_R_eyeAim.Location.X {type=MovieSceneScriptingFloatChannel, default=0, const [0,6]}
       CTRL_R_eyeAim.Location.Y {type=MovieSceneScriptingFloatChannel, default=0, const [0,6]}
       CTRL_R_eyeAim.Location.Z {type=MovieSceneScriptingFloatChannel, default=0, const [0,6]}
       CTRL_R_eyeAim.Rotation.X {type=MovieSceneScriptingFloatChannel, default=0, const [0,6]}
       CTRL_R_eyeAim.Rotation.Y {type=MovieSceneScriptingFloatChannel, default=0, const [0,6]}
       CTRL_R_eyeAim.Rotation.Z {type=MovieSceneScriptingFloatChannel, default=0, const [0,6]}
       CTRL_R_eyeAim.Scale.X {type=MovieSceneScriptingFloatChannel, default=0, const [0,6]}
       CTRL_R_eyeAim.Scale.Y {type=MovieSceneScriptingFloatChannel, default=0, const [0,6]}
       CTRL_R_eyeAim.Scale.Z {type=MovieSceneScriptingFloatChannel, default=0, const [0,6]}
       CTRL_convergenceSwitch {type=MovieSceneScriptingFloatChannel, default=0, const [0,6]}
       CTRL_Master.Location.X {type=MovieSceneScriptingFloatChannel, default=0, const [0,6]}
       CTRL_Master.Location.Y {type=MovieSceneScriptingFloatChannel, default=0, const [0,6]}
       CTRL_Master.Location.Z {type=MovieSceneScriptingFloatChannel, default=0, const [0,6]}
       CTRL_Master.Rotation.X {type=MovieSceneScriptingFloatChannel, default=0, const [0,6]}
       CTRL_Master.Rotation.Y {type=MovieSceneScriptingFloatChannel, default=0, const [0,6]}
       CTRL_Master.Rotation.Z {type=MovieSceneScriptingFloatChannel, default=0, const [0,6]}
       CTRL_Master.Scale.X {type=MovieSceneScriptingFloatChannel, default=0, const [0,6]}
       CTRL_Master.Scale.Y {type=MovieSceneScriptingFloatChannel, default=0, const [0,6]}
       CTRL_Master.Scale.Z {type=MovieSceneScriptingFloatChannel, default=0, const [0,6]}
       CTRL_C_jaw.X {type=MovieSceneScriptingFloatChannel, default=0, thinned_from=6, keys [[130,0],[169,0],[174,-0.0059],[182,-0.0059]]}
       CTRL_C_jaw.Y {type=MovieSceneScriptingFloatChannel, default=0, thinned_from=6, keys [[130,0],[169,0],[174,-0.0492],[182,-0.0492]]}
       CTRL_C_jaw_openExtreme {type=MovieSceneScriptingFloatChannel, default=0, thinned_from=6, keys [[130,0],[169,0],[174,-0.0015],[182,-0.0015]]}
       CTRL_C_jaw_fwdBack {type=MovieSceneScriptingFloatChannel, default=0, thinned_from=6, keys [[130,0],[169,0],[174,-0.0645],[182,-0.0645]]}
       CTRL_R_jaw_ChinRaiseD {type=MovieSceneScriptingFloatChannel, default=0, thinned_from=6, keys [[130,0],[169,0],[174,-0.1288],[182,-0.1288]]}
       CTRL_R_jaw_ChinRaiseU {type=MovieSceneScriptingFloatChannel, default=0, const [0,6]}
       CTRL_L_jaw_ChinRaiseU {type=MovieSceneScriptingFloatChannel, default=0, const [0,6]}
       CTRL_L_jaw_ChinRaiseD {type=MovieSceneScriptingFloatChannel, default=0, thinned_from=6, keys [[130,0],[169,0],[174,-0.1325],[182,-0.1325]]}
       CTRL_R_mouth_lipsBlow {type=MovieSceneScriptingFloatChannel, default=0, const [0,6]}
       CTRL_L_mouth_lipsTogetherU {type=MovieSceneScriptingFloatChannel, default=0, thinned_from=6, keys [[130,0],[169,0],[174,-0.0294],[182,-0.0294]]}
       CTRL_L_mouth_lipsTogetherD {type=MovieSceneScriptingFloatChannel, default=0, thinned_from=6, keys [[130,0],[169,0],[174,-0.0245],[182,-0.0245]]}
       CTRL_R_mouth_lipsTogetherU {type=MovieSceneScriptingFloatChannel, default=0, thinned_from=6, keys [[130,0],[169,0],[174,-0.0294],[182,-0.0294]]}
       CTRL_L_mouth_lipsBlow {type=MovieSceneScriptingFloatChannel, default=0, const [0,6]}
       CTRL_R_mouth_lipsTogetherD {type=MovieSceneScriptingFloatChannel, default=0, thinned_from=6, keys [[130,0],[169,0],[174,-0.0294],[182,-0.0294]]}
       CTRL_C_tongue_roll.X {type=MovieSceneScriptingFloatChannel, default=0, const [0,6]}
       CTRL_C_tongue_roll.Y {type=MovieSceneScriptingFloatChannel, default=0, const [0,6]}
       CTRL_C_tongue.X {type=MovieSceneScriptingFloatChannel, default=0, const [0,6]}
       CTRL_C_tongue.Y {type=MovieSceneScriptingFloatChannel, default=0, const [0,6]}
       CTRL_C_tongue_narrowWide {type=MovieSceneScriptingFloatChannel, default=0, const [0,6]}
       CTRL_C_tongue_inOut {type=MovieSceneScriptingFloatChannel, default=0, const [0,6]}
       CTRL_R_nose.X {type=MovieSceneScriptingFloatChannel, default=0, thinned_from=6, keys [[130,0],[169,0],[174,0.0058],[182,0.0058]]}
       CTRL_R_nose.Y {type=MovieSceneScriptingFloatChannel, default=0, thinned_from=6, keys [[130,0],[169,0],[174,-0.0146],[182,-0.0146]]}
       CTRL_L_nose.X {type=MovieSceneScriptingFloatChannel, default=0, thinned_from=6, keys [[130,0],[169,0],[174,0.0134],[182,0.0134]]}
       CTRL_L_nose.Y {type=MovieSceneScriptingFloatChannel, default=0, thinned_from=6, keys [[130,0],[169,0],[174,0.0242],[182,0.0242]]}
       CTRL_L_eye.X {
        type=MovieSceneScriptingFloatChannel
        default=0
        thinned_from=8
        keys [[72,0],[90,-0.3409],[164,-0.3409],[169,-0.093],[174,0.1636],[182,0.1636]]
       }
       CTRL_L_eye.Y {
        type=MovieSceneScriptingFloatChannel
        default=0
        thinned_from=8
        keys [[72,0],[90,-0.149],[130,-0.149],[164,-0.1151],[169,-0.5456],[174,-0.7856],[182,-0.7856]]
       }
       CTRL_L_eye_blink {
        type=MovieSceneScriptingFloatChannel
        default=0
        keys [[130,0],[164,0],[166,1.4777],[169,0],[174,0.4732],[178,0.5173],[180,1.5336],[182,0.5329]]
       }
       CTRL_R_eye.X {
        type=MovieSceneScriptingFloatChannel
        default=0
        thinned_from=8
        keys [[72,0],[90,-0.3409],[164,-0.3409],[169,-0.093],[174,0.1773],[180,0.1733],[182,0.1733]]
       }
       CTRL_R_eye.Y {
        type=MovieSceneScriptingFloatChannel
        default=0
        thinned_from=8
        keys [[72,0],[90,-0.1505],[130,-0.1505],[164,-0.1154],[169,-0.5456],[174,-0.7866],[182,-0.7866]]
       }
       CTRL_R_eye_blink {
        type=MovieSceneScriptingFloatChannel
        default=0
        keys [[130,0],[164,0],[166,1.6085],[169,0],[174,0.6164],[178,0.6796],[180,1.6855],[182,0.688]]
       }
       CTRL_C_eye.X {
        type=MovieSceneScriptingFloatChannel
        default=0.3646
        thinned_from=10
        keys [[60,0.0328],[90,0.4678],[164,0.4678],[169,-0.3487],[174,-0.6871],[178,-0.6871],[180,-0.4838],[182,-0.181],[183,-0.1097]]
       }
       CTRL_C_eye.Y {
        type=MovieSceneScriptingFloatChannel
        default=-1
        thinned_from=10
        keys [[60,-0.249],[90,-0.8285],[164,-0.8285],[169,-0.6873],[174,-0.5279],[178,-0.5279],[180,-0.1146],[182,0.1279],[183,0.2204]]
       }
       CTRL_C_eyesConvergence {type=MovieSceneScriptingFloatChannel, default=0, const [0,6]}
       CTRL_L_eye_cheekRaise {type=MovieSceneScriptingFloatChannel, default=0, thinned_from=6, keys [[130,0],[169,0],[174,-0.019],[182,-0.019]]}
       CTRL_L_eye_squintInner {type=MovieSceneScriptingFloatChannel, default=0, thinned_from=6, keys [[130,0],[169,0],[174,-0.0632],[182,-0.0632]]}
       CTRL_L_nose_wrinkleUpper {type=MovieSceneScriptingFloatChannel, default=0, thinned_from=6, keys [[130,0],[169,0],[174,-1],[182,-1]]}
       CTRL_R_nose_wrinkleUpper {type=MovieSceneScriptingFloatChannel, default=0, thinned_from=6, keys [[130,0],[169,0],[174,-1],[182,-1]]}
       CTRL_R_eye_squintInner {type=MovieSceneScriptingFloatChannel, default=0, thinned_from=6, keys [[130,0],[169,0],[174,-0.0614],[182,-0.0614]]}
       CTRL_R_eye_cheekRaise {type=MovieSceneScriptingFloatChannel, default=0, const [0,6]}
       CTRL_C_eyelidsReduce {type=MovieSceneScriptingFloatChannel, default=0, const [0,6]}
       CTRL_R_brow_lateral {type=MovieSceneScriptingFloatChannel, default=0, thinned_from=6, keys [[130,0],[169,0],[174,-0.0011],[180,-0.0011],[182,0]]}
       CTRL_L_brow_lateral {type=MovieSceneScriptingFloatChannel, default=0, thinned_from=6, keys [[130,0],[169,0],[174,-0.0046],[180,-0.0046],[182,0]]}
       CTRL_R_brow_down {type=MovieSceneScriptingFloatChannel, default=0, thinned_from=6, keys [[130,0],[169,0],[174,-0.0019],[180,-0.0019],[182,0]]}
       CTRL_L_brow_down {type=MovieSceneScriptingFloatChannel, default=0, thinned_from=6, keys [[130,0],[169,0],[174,-0.004],[180,-0.004],[182,0]]}
       CTRL_L_brow_raiseOut {type=MovieSceneScriptingFloatChannel, default=0, thinned_from=6, keys [[130,0],[169,0],[174,-0.841],[180,-0.841],[182,0]]}
       CTRL_L_brow_raiseIn {type=MovieSceneScriptingFloatChannel, default=0, thinned_from=6, keys [[130,0],[169,0],[174,-0.5337],[180,-0.5337],[182,0]]}
       CTRL_R_brow_raiseIn {type=MovieSceneScriptingFloatChannel, default=0, thinned_from=6, keys [[130,0],[169,0],[174,-0.5148],[180,-0.5148],[182,0]]}
       CTRL_R_brow_raiseOut {type=MovieSceneScriptingFloatChannel, default=0, thinned_from=6, keys [[130,0],[169,0],[174,-0.7905],[180,-0.7905],[182,0]]}
       CTRL_lookAtSwitch {type=MovieSceneScriptingFloatChannel, default=0, const [0,6]}
       CTRL_C_teethD.X {type=MovieSceneScriptingFloatChannel, default=0, const [0,6]}
       CTRL_C_teethD.Y {type=MovieSceneScriptingFloatChannel, default=0, const [0,6]}
       CTRL_C_teethU.X {type=MovieSceneScriptingFloatChannel, default=0, const [0,6]}
       CTRL_C_teethU.Y {type=MovieSceneScriptingFloatChannel, default=0, const [0,6]}
       CTRL_C_teeth_fwdBackU {type=MovieSceneScriptingFloatChannel, default=0, const [0,6]}
       CTRL_L_mouth_corner.X {type=MovieSceneScriptingFloatChannel, default=0, const [0,6]}
       CTRL_L_mouth_corner.Y {type=MovieSceneScriptingFloatChannel, default=0, thinned_from=6, keys [[130,0],[169,0],[174,-0.0368],[182,-0.0368]]}
       CTRL_R_mouth_corner.X {type=MovieSceneScriptingFloatChannel, default=0, const [0,6]}
       CTRL_R_mouth_corner.Y {type=MovieSceneScriptingFloatChannel, default=0, thinned_from=6, keys [[130,0],[169,0],[174,-0.0251],[182,-0.0251]]}
       CTRL_C_teeth_fwdBackD {type=MovieSceneScriptingFloatChannel, default=0, const [0,6]}
       CTRL_R_mouth_cornerSharpnessD {type=MovieSceneScriptingFloatChannel, default=0, thinned_from=6, keys [[130,0],[169,0],[174,-0.0002],[182,-0.0002]]}
       CTRL_R_mouth_cornerSharpnessU {type=MovieSceneScriptingFloatChannel, default=0, thinned_from=6, keys [[130,0],[169,0],[174,-0.0073],[182,-0.0073]]}
       CTRL_L_eye_eyelidU.X {type=MovieSceneScriptingFloatChannel, default=0, const [0,6]}
       CTRL_L_eye_eyelidU.Y {type=MovieSceneScriptingFloatChannel, default=0, const [0,6]}
       CTRL_L_mouth_cornerSharpnessD {type=MovieSceneScriptingFloatChannel, default=0, thinned_from=6, keys [[130,0],[169,0],[174,-0.0103],[182,-0.0103]]}
       CTRL_L_eye_eyelidD.X {type=MovieSceneScriptingFloatChannel, default=0, const [0,6]}
       CTRL_L_eye_eyelidD.Y {type=MovieSceneScriptingFloatChannel, default=0, const [0,6]}
       CTRL_L_mouth_cornerSharpnessU {type=MovieSceneScriptingFloatChannel, default=0, thinned_from=6, keys [[130,0],[169,0],[174,-0.0142],[182,-0.0142]]}
       CTRL_R_mouth_lipsTowardsTeethD {type=MovieSceneScriptingFloatChannel, default=0, const [0,6]}
       CTRL_L_mouth_lipsTowardsTeethD {type=MovieSceneScriptingFloatChannel, default=0, const [0,6]}
       CTRL_R_eye_eyelidU.X {type=MovieSceneScriptingFloatChannel, default=0, const [0,6]}
       CTRL_R_eye_eyelidU.Y {type=MovieSceneScriptingFloatChannel, default=0, const [0,6]}
       CTRL_L_mouth_lipsTowardsTeethU {type=MovieSceneScriptingFloatChannel, default=0, const [0,6]}
       CTRL_R_eye_eyelidD.X {type=MovieSceneScriptingFloatChannel, default=0, const [0,6]}
       CTRL_R_eye_eyelidD.Y {type=MovieSceneScriptingFloatChannel, default=0, const [0,6]}
       CTRL_R_mouth_lipsTowardsTeethU {type=MovieSceneScriptingFloatChannel, default=0, const [0,6]}
       CTRL_faceGUIfollowHead {type=MovieSceneScriptingFloatChannel, default=0, const [0,6]}
       CTRL_eyesAimFollowHead {type=MovieSceneScriptingFloatChannel, default=0, const [0,6]}
       CTRL_C_nose.X {type=MovieSceneScriptingFloatChannel, default=0, const [0,6]}
       CTRL_C_nose.Y {type=MovieSceneScriptingFloatChannel, default=0, const [0,6]}
       CTRL_C_teeth_rotateU {type=MovieSceneScriptingFloatChannel, default=0, const [0,6]}
       CTRL_C_teeth_rotateD {type=MovieSceneScriptingFloatChannel, default=0, const [0,6]}
       CTRL_L_brow_inner {type=MovieSceneScriptingFloatChannel, default=0, const [0,6]}
       CTRL_R_brow_inner {type=MovieSceneScriptingFloatChannel, default=0, const [0,6]}
       CTRL_L_brow_outer {type=MovieSceneScriptingFloatChannel, default=0, const [0,6]}
       CTRL_R_brow_outer {type=MovieSceneScriptingFloatChannel, default=0, const [0,6]}
       CTRL_C_mouth.X {type=MovieSceneScriptingFloatChannel, default=0, thinned_from=6, keys [[130,0],[169,0],[174,-0.0215],[182,-0.0215]]}
       CTRL_C_mouth.Y {type=MovieSceneScriptingFloatChannel, default=0, const [0,6]}
       CTRL_L_mouth_upperLipRaise {type=MovieSceneScriptingFloatChannel, default=0, thinned_from=6, keys [[130,0],[169,0],[174,-0.0053],[182,-0.0053]]}
       CTRL_L_mouth_cornerPull {
        type=MovieSceneScriptingFloatChannel
        default=0.1087
        thinned_from=8
        keys [[72,0.5279],[130,0.2096],[148,0.4007],[164,0.4007],[169,0.1087],[174,-0.2948],[182,-0.2948]]
       }
       CTRL_L_mouth_dimple {type=MovieSceneScriptingFloatChannel, default=0, thinned_from=8, keys [[72,0.6941],[169,0.6941],[174,-0.2888],[182,-0.2888]]}
       CTRL_L_mouth_cornerDepress {type=MovieSceneScriptingFloatChannel, default=0, thinned_from=6, keys [[130,0],[169,0],[174,-0.0035],[182,-0.0035]]}
       CTRL_L_mouth_stretch {type=MovieSceneScriptingFloatChannel, default=0, thinned_from=6, keys [[130,0],[169,0],[174,-0.0288],[182,-0.0288]]}
       CTRL_L_mouth_lowerLipDepress {type=MovieSceneScriptingFloatChannel, default=0, thinned_from=6, keys [[130,0],[169,0],[174,-0.0013],[182,-0.0013]]}
       CTRL_R_mouth_lowerLipDepress {type=MovieSceneScriptingFloatChannel, default=0, thinned_from=6, keys [[130,0],[169,0],[174,-0.0018],[182,-0.0018]]}
       CTRL_R_mouth_stretch {type=MovieSceneScriptingFloatChannel, default=0, thinned_from=6, keys [[130,0],[169,0],[174,-0.0229],[182,-0.0229]]}
       CTRL_L_mouth_towardsD {type=MovieSceneScriptingFloatChannel, default=0, thinned_from=6, keys [[130,0.0001],[169,0.0001],[174,-0.0029],[182,-0.0029]]}
       CTRL_R_mouth_funnelU {type=MovieSceneScriptingFloatChannel, default=0, thinned_from=6, keys [[130,0],[169,0],[174,-0.0037],[182,-0.0037]]}
       CTRL_L_mouth_towardsU {type=MovieSceneScriptingFloatChannel, default=0, thinned_from=6, keys [[130,0],[169,0],[174,-0.0088],[182,-0.0088]]}
       CTRL_R_mouth_lipBiteD {type=MovieSceneScriptingFloatChannel, default=0, thinned_from=6, keys [[130,0],[169,0],[174,-0.0144],[182,-0.0144]]}
       CTRL_R_mouth_funnelD {type=MovieSceneScriptingFloatChannel, default=0, thinned_from=6, keys [[130,0],[169,0],[174,-0.0018],[182,-0.0018]]}
       CTRL_L_mouth_pressU {type=MovieSceneScriptingFloatChannel, default=0, const [0,6]}
       CTRL_L_mouth_funnelD {type=MovieSceneScriptingFloatChannel, default=0, thinned_from=6, keys [[130,0],[169,0],[174,-0.0008],[182,-0.0008]]}
       CTRL_L_mouth_suckBlow {type=MovieSceneScriptingFloatChannel, default=0, thinned_from=6, keys [[130,0],[169,0],[174,-0.0049],[182,-0.0049]]}
       CTRL_L_mouth_pressD {type=MovieSceneScriptingFloatChannel, default=0, const [0,6]}
       CTRL_L_mouth_funnelU {type=MovieSceneScriptingFloatChannel, default=0, thinned_from=6, keys [[130,0],[169,0],[174,-0.0029],[182,-0.0029]]}
       CTRL_R_mouth_towardsU {type=MovieSceneScriptingFloatChannel, default=-0.0001, thinned_from=6, keys [[130,-0.0001],[169,-0.0001],[174,-0.013],[182,-0.013]]}
       CTRL_R_mouth_purseD {type=MovieSceneScriptingFloatChannel, default=0, thinned_from=6, keys [[130,0],[169,0],[174,-0.0191],[182,-0.0191]]}
       CTRL_L_mouth_lipBiteU {type=MovieSceneScriptingFloatChannel, default=0, thinned_from=6, keys [[130,0],[169,0],[174,-0.0253],[182,-0.0253]]}
       CTRL_R_mouth_pressD {type=MovieSceneScriptingFloatChannel, default=0, const [0,6]}
       CTRL_R_mouth_pressU {type=MovieSceneScriptingFloatChannel, default=0, const [0,6]}
       CTRL_L_mouth_purseD {type=MovieSceneScriptingFloatChannel, default=0, thinned_from=6, keys [[130,0],[169,0],[174,-0.0372],[182,-0.0372]]}
       CTRL_R_mouth_lipBiteU {type=MovieSceneScriptingFloatChannel, default=0, thinned_from=6, keys [[130,0],[169,0],[174,-0.02],[182,-0.02]]}
       CTRL_R_mouth_suckBlow {type=MovieSceneScriptingFloatChannel, default=0, thinned_from=6, keys [[130,0],[169,0],[174,-0.0117],[182,-0.0117]]}
       CTRL_L_mouth_purseU {type=MovieSceneScriptingFloatChannel, default=0, thinned_from=6, keys [[130,0],[169,0],[174,-0.0499],[182,-0.0499]]}
       CTRL_R_mouth_towardsD {type=MovieSceneScriptingFloatChannel, default=0, thinned_from=6, keys [[130,0],[169,0],[174,-0.0025],[182,-0.0025]]}
       CTRL_L_mouth_lipBiteD {type=MovieSceneScriptingFloatChannel, default=0, thinned_from=6, keys [[130,0],[169,0],[174,-0.0149],[182,-0.0149]]}
       CTRL_R_mouth_purseU {type=MovieSceneScriptingFloatChannel, default=0, thinned_from=6, keys [[130,0],[169,0],[174,-0.095],[182,-0.095]]}
       CTRL_R_mouth_cornerDepress {type=MovieSceneScriptingFloatChannel, default=0, thinned_from=6, keys [[130,0],[169,0],[174,-0.002],[182,-0.002]]}
       CTRL_R_mouth_upperLipRaise {type=MovieSceneScriptingFloatChannel, default=0, thinned_from=6, keys [[130,0],[169,0],[174,-0.0013],[182,-0.0013]]}
       CTRL_R_mouth_dimple {
        type=MovieSceneScriptingFloatChannel
        default=0.8943
        thinned_from=9
        keys [[52,0.8986],[164,0.8986],[169,0.8943],[174,-0.1057],[182,-0.1057]]
       }
       CTRL_R_mouth_cornerPull {
        type=MovieSceneScriptingFloatChannel
        default=0
        thinned_from=9
        keys [[52,0],[72,0.0521],[130,0.0521],[148,0.5318],[164,0.563],[169,0],[174,-0.0314],[182,-0.0315]]
       }
      ]
     }
    ]
   }
   {
    type=MovieSceneControlRigParameterTrack
    display_name=Fortnite_Biped_CtrlRig
    sections [
     {
      type=MovieSceneControlRigParameterSection
      channels [
       leg_l_stretch_switch {
        type=MovieSceneScriptingBoolChannel
        default=false
        keys [[0,false],[1,false],[2,false],[3,false],[4,false],[5,false],[6,false],[7,false],[8,false],[9,false],[10,false],[11,false],[12,false],[13,false],[14,false],[15,false],[16,false],[17,false],[18,false],[19,false],[20,false],[21,false],[22,false],[23,false],[24,false],[25,false],[26,false],[27,false],[28,false],[29,false],[30,false],[31,false],[32,false],[33,false],[34,false],[35,false],[36,false],[37,false],[38,false],[39,false],[40,false],[41,false],[42,false],[43,false],[44,false],[45,false],[46,false],[47,false],[48,false],[49,false],[50,false],[51,false],[52,false],[53,false],[54,false],[55,false],[56,false],[57,false],[58,false],[59,false],[60,false],[61,false],[62,false],[63,false],[64,false],[65,false],[66,false],[67,false],[68,false],[69,false],[70,false],[71,false],[72,false],[73,false],[74,false],[75,false],[76,false],[77,false],[78,false],[79,false],[80,false],[81,false],[82,false],[83,false],[84,false],[85,false],[86,false],[87,false],[88,false],[89,false],[90,false],[91,false],[92,false],[93,false],[94,false],[95,false],[96,false],[97,false],[98,false],[99,false],[100,false],[101,false],[102,false],[103,false],[104,false],[105,false],[106,false],[107,false],[108,false],[109,false],[110,false],[111,false],[112,false],[113,false],[114,false],[115,false],[116,false],[117,false],[118,false],[119,false],[120,false],[121,false],[122,false],[123,false],[124,false],[125,false],[126,false],[127,false],[128,false],[129,false],[130,false],[131,false],[132,false],[133,false],[134,false],[135,false],[136,false],[137,false],[138,false],[139,false],[140,false],[141,false],[142,false],[143,false],[144,false],[145,false],[146,false],[147,false],[148,false],[149,false],[150,false],[151,false],[152,false],[153,false],[154,false],[155,false],[156,false],[157,false],[158,false],[159,false],[160,false],[161,false],[162,false],[163,false],[164,false],[165,false],[166,false],[167,false],[168,false],[169,false],[170,false],[171,false],[172,false],[173,true],[174,false],[175,true],[176,false],[177,false],[178,true],[179,false],[180,true],[181,true],[182,true],[183,false],[184,false]]
       }
       leg_r_stretch_switch {
        type=MovieSceneScriptingBoolChannel
        default=false
        keys [[0,false],[1,false],[2,false],[3,false],[4,false],[5,false],[6,false],[7,false],[8,false],[9,false],[10,false],[11,false],[12,false],[13,false],[14,false],[15,false],[16,false],[17,false],[18,false],[19,false],[20,false],[21,false],[22,false],[23,false],[24,false],[25,false],[26,false],[27,false],[28,false],[29,false],[30,false],[31,false],[32,false],[33,false],[34,false],[35,false],[36,false],[37,false],[38,false],[39,false],[40,false],[41,false],[42,false],[43,false],[44,false],[45,false],[46,false],[47,false],[48,false],[49,false],[50,false],[51,false],[52,false],[53,false],[54,false],[55,false],[56,false],[57,false],[58,false],[59,false],[60,false],[61,false],[62,false],[63,false],[64,false],[65,false],[66,false],[67,false],[68,false],[69,false],[70,false],[71,false],[72,false],[73,false],[74,false],[75,false],[76,false],[77,false],[78,false],[79,false],[80,false],[81,false],[82,false],[83,false],[84,false],[85,false],[86,false],[87,false],[88,false],[89,false],[90,false],[91,false],[92,false],[93,false],[94,false],[95,false],[96,false],[97,false],[98,false],[99,false],[100,false],[101,false],[102,false],[103,false],[104,false],[105,false],[106,false],[107,false],[108,false],[109,false],[110,false],[111,false],[112,false],[113,false],[114,false],[115,false],[116,false],[117,false],[118,false],[119,false],[120,false],[121,false],[122,false],[123,false],[124,false],[125,false],[126,false],[127,false],[128,false],[129,false],[130,false],[131,false],[132,false],[133,false],[134,false],[135,false],[136,false],[137,false],[138,false],[139,false],[140,false],[141,false],[142,false],[143,false],[144,false],[145,false],[146,false],[147,false],[148,false],[149,false],[150,false],[151,false],[152,false],[153,false],[154,false],[155,false],[156,false],[157,false],[158,false],[159,false],[160,false],[161,false],[162,false],[163,false],[164,false],[165,false],[166,false],[167,false],[168,false],[169,false],[170,false],[171,false],[172,false],[173,true],[174,false],[175,true],[176,false],[177,false],[178,true],[179,false],[180,true],[181,true],[182,true],[183,false],[184,false]]
       }
       arm_l_fk_local {type=MovieSceneScriptingBoolChannel, default=true, const [true,178]}
       Auto Interpolate_3 {
        type=MovieSceneScriptingBoolChannel
        default=false
        keys [[0,false],[1,false],[2,false],[3,false],[4,false],[5,false],[6,false],[7,false],[8,false],[9,false],[10,false],[11,false],[12,false],[13,false],[14,false],[15,false],[16,false],[17,false],[18,false],[19,false],[20,false],[21,false],[22,false],[23,false],[24,false],[25,false],[26,false],[27,false],[28,false],[29,false],[30,false],[31,false],[32,false],[33,false],[34,false],[35,false],[36,false],[37,false],[38,false],[39,false],[40,false],[41,false],[42,false],[43,false],[44,false],[45,false],[46,false],[47,false],[48,false],[49,false],[50,false],[51,false],[52,false],[53,false],[54,false],[55,false],[56,false],[57,false],[58,false],[59,false],[60,false],[61,false],[62,false],[63,false],[64,false],[65,false],[66,false],[67,false],[68,false],[69,false],[70,false],[71,false],[72,false],[73,false],[74,false],[75,false],[76,false],[77,false],[78,false],[79,false],[80,false],[81,false],[82,false],[83,false],[84,false],[85,false],[86,false],[87,false],[88,false],[89,false],[90,false],[91,false],[92,false],[93,false],[94,false],[95,false],[96,false],[97,false],[98,false],[99,false],[100,false],[101,false],[102,false],[103,false],[104,false],[105,false],[106,false],[107,false],[108,false],[109,false],[110,false],[111,false],[112,false],[113,false],[114,false],[115,false],[116,false],[117,false],[118,false],[119,false],[120,false],[121,false],[122,false],[123,false],[124,false],[125,false],[126,false],[127,false],[128,false],[129,false],[130,false],[131,false],[132,false],[133,false],[134,false],[135,false],[136,false],[137,false],[138,false],[139,false],[140,false],[141,false],[142,false],[143,false],[144,false],[145,false],[146,false],[147,false],[148,false],[149,false],[150,false],[151,false],[152,false],[153,false],[154,false],[155,false],[156,false],[157,false],[158,false],[159,false],[160,false],[161,false],[162,false],[163,false],[164,false],[165,false],[166,false],[167,false],[168,false],[169,false],[170,false],[171,false],[172,false],[173,true],[174,false],[175,true],[176,false],[177,false],[178,true],[179,false],[180,true],[181,true],[182,true],[183,false],[184,false]]
       }
       Auto Interpolate {
        type=MovieSceneScriptingBoolChannel
        default=false
        keys [[0,false],[1,false],[2,false],[3,false],[4,false],[5,false],[6,false],[7,false],[8,false],[9,false],[10,false],[11,false],[12,false],[13,false],[14,false],[15,false],[16,false],[17,false],[18,false],[19,false],[20,false],[21,false],[22,false],[23,false],[24,false],[25,false],[26,false],[27,false],[28,false],[29,false],[30,false],[31,false],[32,false],[33,false],[34,false],[35,false],[36,false],[37,false],[38,false],[39,false],[40,false],[41,false],[42,false],[43,false],[44,false],[45,false],[46,false],[47,false],[48,false],[49,false],[50,false],[51,false],[52,false],[53,false],[54,false],[55,false],[56,false],[57,false],[58,false],[59,false],[60,false],[61,false],[62,false],[63,false],[64,false],[65,false],[66,false],[67,false],[68,false],[69,false],[70,false],[71,false],[72,false],[73,false],[74,false],[75,false],[76,false],[77,false],[78,false],[79,false],[80,false],[81,false],[82,false],[83,false],[84,false],[85,false],[86,false],[87,false],[88,false],[89,false],[90,false],[91,false],[92,false],[93,false],[94,false],[95,false],[96,false],[97,false],[98,false],[99,false],[100,false],[101,false],[102,false],[103,false],[104,false],[105,false],[106,false],[107,false],[108,false],[109,false],[110,false],[111,false],[112,false],[113,false],[114,false],[115,false],[116,false],[117,false],[118,false],[119,false],[120,false],[121,false],[122,false],[123,false],[124,false],[125,false],[126,false],[127,false],[128,false],[129,false],[130,false],[131,false],[132,false],[133,false],[134,false],[135,false],[136,false],[137,false],[138,false],[139,false],[140,false],[141,false],[142,false],[143,false],[144,false],[145,false],[146,false],[147,false],[148,false],[149,false],[150,false],[151,false],[152,false],[153,false],[154,false],[155,false],[156,false],[157,false],[158,false],[159,false],[160,false],[161,false],[162,false],[163,false],[164,false],[165,false],[166,false],[167,false],[168,false],[169,false],[170,false],[171,false],[172,false],[173,true],[174,false],[175,true],[176,false],[177,false],[178,true],[179,false],[180,true],[181,true],[182,true],[183,false],[184,false]]
       }
       arm_l_stretch_switch {
        type=MovieSceneScriptingBoolChannel
        default=false
        keys [[0,false],[1,false],[2,false],[3,false],[4,false],[5,false],[6,false],[7,false],[8,false],[9,false],[10,false],[11,false],[12,false],[13,false],[14,false],[15,false],[16,false],[17,false],[18,false],[19,false],[20,false],[21,false],[22,false],[23,false],[24,false],[25,false],[26,false],[27,false],[28,false],[29,false],[30,false],[31,false],[32,false],[33,false],[34,false],[35,false],[36,false],[37,false],[38,false],[39,false],[40,false],[41,false],[42,false],[43,false],[44,false],[45,false],[46,false],[47,false],[48,false],[49,false],[50,false],[51,false],[52,false],[53,false],[54,false],[55,false],[56,false],[57,false],[58,false],[59,false],[60,false],[61,false],[62,false],[63,false],[64,false],[65,false],[66,false],[67,false],[68,false],[69,false],[70,false],[71,false],[72,false],[73,false],[74,false],[75,false],[76,false],[77,false],[78,false],[79,false],[80,false],[81,false],[82,false],[83,false],[84,false],[85,false],[86,false],[87,false],[88,false],[89,false],[90,false],[91,false],[92,false],[93,false],[94,false],[95,false],[96,false],[97,false],[98,false],[99,false],[100,false],[101,false],[102,false],[103,false],[104,false],[105,false],[106,false],[107,false],[108,false],[109,false],[110,false],[111,false],[112,false],[113,false],[114,false],[115,false],[116,false],[117,false],[118,false],[119,false],[120,false],[121,false],[122,false],[123,false],[124,false],[125,false],[126,false],[127,false],[128,false],[129,false],[130,false],[131,false],[132,false],[133,false],[134,false],[135,false],[136,false],[137,false],[138,false],[139,false],[140,false],[141,false],[142,false],[143,false],[144,false],[145,false],[146,false],[147,false],[148,false],[149,false],[150,false],[151,false],[152,false],[153,false],[154,false],[155,false],[156,false],[157,false],[158,false],[159,false],[160,false],[161,false],[162,false],[163,false],[164,false],[165,false],[166,false],[167,false],[168,false],[169,false],[170,false],[171,false],[172,false],[173,true],[174,false],[175,true],[176,false],[177,false],[178,true],[179,false],[180,true],[181,true],[182,true],[183,false],[184,false]]
       }
       Keep Aligned {
        type=MovieSceneScriptingBoolChannel
        default=false
        keys [[0,false],[1,false],[2,false],[3,false],[4,false],[5,false],[6,false],[7,false],[8,false],[9,false],[10,false],[11,false],[12,false],[13,false],[14,false],[15,false],[16,false],[17,false],[18,false],[19,false],[20,false],[21,false],[22,false],[23,false],[24,false],[25,false],[26,false],[27,false],[28,false],[29,false],[30,false],[31,false],[32,false],[33,false],[34,false],[35,false],[36,false],[37,false],[38,false],[39,false],[40,false],[41,false],[42,false],[43,false],[44,false],[45,false],[46,false],[47,false],[48,false],[49,false],[50,false],[51,false],[52,false],[53,false],[54,false],[55,false],[56,false],[57,false],[58,false],[59,false],[60,false],[61,false],[62,false],[63,false],[64,false],[65,false],[66,false],[67,false],[68,false],[69,false],[70,false],[71,false],[72,false],[73,false],[74,false],[75,false],[76,false],[77,false],[78,false],[79,false],[80,false],[81,false],[82,false],[83,false],[84,false],[85,false],[86,false],[87,false],[88,false],[89,false],[90,false],[91,false],[92,false],[93,false],[94,false],[95,false],[96,false],[97,false],[98,false],[99,false],[100,false],[101,false],[102,false],[103,false],[104,false],[105,false],[106,false],[107,false],[108,false],[109,false],[110,false],[111,false],[112,false],[113,false],[114,false],[115,false],[116,false],[117,false],[118,false],[119,false],[120,false],[121,false],[122,false],[123,false],[124,false],[125,false],[126,false],[127,false],[128,false],[129,false],[130,false],[131,false],[132,false],[133,false],[134,false],[135,false],[136,false],[137,false],[138,false],[139,false],[140,false],[141,false],[142,false],[143,false],[144,false],[145,false],[146,false],[147,false],[148,false],[149,false],[150,false],[151,false],[152,false],[153,false],[154,false],[155,false],[156,false],[157,false],[158,false],[159,false],[160,false],[161,false],[162,false],[163,false],[164,false],[165,false],[166,false],[167,false],[168,false],[169,false],[170,false],[171,false],[172,false],[173,true],[174,false],[175,true],[176,false],[177,false],[178,true],[179,false],[180,true],[181,true],[182,true],[183,false],[184,false]]
       }
       arm_r_fk_local {type=MovieSceneScriptingBoolChannel, default=true, const [true,175]}
       Auto Interpolate_4 {
        type=MovieSceneScriptingBoolChannel
        default=false
        keys [[0,false],[1,false],[2,false],[3,false],[4,false],[5,false],[6,false],[7,false],[8,false],[9,false],[10,false],[11,false],[12,false],[13,false],[14,false],[15,false],[16,false],[17,false],[18,false],[19,false],[20,false],[21,false],[22,false],[23,false],[24,false],[25,false],[26,false],[27,false],[28,false],[29,false],[30,false],[31,false],[32,false],[33,false],[34,false],[35,false],[36,false],[37,false],[38,false],[39,false],[40,false],[41,false],[42,false],[43,false],[44,false],[45,false],[46,false],[47,false],[48,false],[49,false],[50,false],[51,false],[52,false],[53,false],[54,false],[55,false],[56,false],[57,false],[58,false],[59,false],[60,false],[61,false],[62,false],[63,false],[64,false],[65,false],[66,false],[67,false],[68,false],[69,false],[70,false],[71,false],[72,false],[73,false],[74,false],[75,false],[76,false],[77,false],[78,false],[79,false],[80,false],[81,false],[82,false],[83,false],[84,false],[85,false],[86,false],[87,false],[88,false],[89,false],[90,false],[91,false],[92,false],[93,false],[94,false],[95,false],[96,false],[97,false],[98,false],[99,false],[100,false],[101,false],[102,false],[103,false],[104,false],[105,false],[106,false],[107,false],[108,false],[109,false],[110,false],[111,false],[112,false],[113,false],[114,false],[115,false],[116,false],[117,false],[118,false],[119,false],[120,false],[121,false],[122,false],[123,false],[124,false],[125,false],[126,false],[127,false],[128,false],[129,false],[130,false],[131,false],[132,false],[133,false],[134,false],[135,false],[136,false],[137,false],[138,false],[139,false],[140,false],[141,false],[142,false],[143,false],[144,false],[145,false],[146,false],[147,false],[148,false],[149,false],[150,false],[151,false],[152,false],[153,false],[154,false],[155,false],[156,false],[157,false],[158,false],[159,false],[160,false],[161,false],[162,false],[163,false],[164,false],[165,false],[166,false],[167,false],[168,false],[169,false],[170,false],[171,false],[172,false],[173,true],[174,false],[175,true],[176,false],[177,false],[178,true],[179,false],[180,true],[181,true],[182,true],[183,false],[184,false]]
       }
       Auto Interpolate_2 {
        type=MovieSceneScriptingBoolChannel
        default=false
        keys [[0,false],[1,false],[2,false],[3,false],[4,false],[5,false],[6,false],[7,false],[8,false],[9,false],[10,false],[11,false],[12,false],[13,false],[14,false],[15,false],[16,false],[17,false],[18,false],[19,false],[20,false],[21,false],[22,false],[23,false],[24,false],[25,false],[26,false],[27,false],[28,false],[29,false],[30,false],[31,false],[32,false],[33,false],[34,false],[35,false],[36,false],[37,false],[38,false],[39,false],[40,false],[41,false],[42,false],[43,false],[44,false],[45,false],[46,false],[47,false],[48,false],[49,false],[50,false],[51,false],[52,false],[53,false],[54,false],[55,false],[56,false],[57,false],[58,false],[59,false],[60,false],[61,false],[62,false],[63,false],[64,false],[65,false],[66,false],[67,false],[68,false],[69,false],[70,false],[71,false],[72,false],[73,false],[74,false],[75,false],[76,false],[77,false],[78,false],[79,false],[80,false],[81,false],[82,false],[83,false],[84,false],[85,false],[86,false],[87,false],[88,false],[89,false],[90,false],[91,false],[92,false],[93,false],[94,false],[95,false],[96,false],[97,false],[98,false],[99,false],[100,false],[101,false],[102,false],[103,false],[104,false],[105,false],[106,false],[107,false],[108,false],[109,false],[110,false],[111,false],[112,false],[113,false],[114,false],[115,false],[116,false],[117,false],[118,false],[119,false],[120,false],[121,false],[122,false],[123,false],[124,false],[125,false],[126,false],[127,false],[128,false],[129,false],[130,false],[131,false],[132,false],[133,false],[134,false],[135,false],[136,false],[137,false],[138,false],[139,false],[140,false],[141,false],[142,false],[143,false],[144,false],[145,false],[146,false],[147,false],[148,false],[149,false],[150,false],[151,false],[152,false],[153,false],[154,false],[155,false],[156,false],[157,false],[158,false],[159,false],[160,false],[161,false],[162,false],[163,false],[164,false],[165,false],[166,false],[167,false],[168,false],[169,false],[170,false],[171,false],[172,false],[173,true],[174,false],[175,true],[176,false],[177,false],[178,true],[179,false],[180,true],[181,true],[182,true],[183,false],[184,false]]
       }
       arm_r_stretch_switch {
        type=MovieSceneScriptingBoolChannel
        default=false
        keys [[0,false],[1,false],[2,false],[3,false],[4,false],[5,false],[6,false],[7,false],[8,false],[9,false],[10,false],[11,false],[12,false],[13,false],[14,false],[15,false],[16,false],[17,false],[18,false],[19,false],[20,false],[21,false],[22,false],[23,false],[24,false],[25,false],[26,false],[27,false],[28,false],[29,false],[30,false],[31,false],[32,false],[33,false],[34,false],[35,false],[36,false],[37,false],[38,false],[39,false],[40,false],[41,false],[42,false],[43,false],[44,false],[45,false],[46,false],[47,false],[48,false],[49,false],[50,false],[51,false],[52,false],[53,false],[54,false],[55,false],[56,false],[57,false],[58,false],[59,false],[60,false],[61,false],[62,false],[63,false],[64,false],[65,false],[66,false],[67,false],[68,false],[69,false],[70,false],[71,false],[72,false],[73,false],[74,false],[75,false],[76,false],[77,false],[78,false],[79,false],[80,false],[81,false],[82,false],[83,false],[84,false],[85,false],[86,false],[87,false],[88,false],[89,false],[90,false],[91,false],[92,false],[93,false],[94,false],[95,false],[96,false],[97,false],[98,false],[99,false],[100,false],[101,false],[102,false],[103,false],[104,false],[105,false],[106,false],[107,false],[108,false],[109,false],[110,false],[111,false],[112,false],[113,false],[114,false],[115,false],[116,false],[117,false],[118,false],[119,false],[120,false],[121,false],[122,false],[123,false],[124,false],[125,false],[126,false],[127,false],[128,false],[129,false],[130,false],[131,false],[132,false],[133,false],[134,false],[135,false],[136,false],[137,false],[138,false],[139,false],[140,false],[141,false],[142,false],[143,false],[144,false],[145,false],[146,false],[147,false],[148,false],[149,false],[150,false],[151,false],[152,false],[153,false],[154,false],[155,false],[156,false],[157,false],[158,false],[159,false],[160,false],[161,false],[162,false],[163,false],[164,false],[165,false],[166,false],[167,false],[168,false],[169,false],[170,false],[171,false],[172,false],[173,true],[174,false],[175,true],[176,false],[177,false],[178,true],[179,false],[180,true],[181,true],[182,true],[183,false],[184,false]]
       }
       Keep Aligned_2 {
        type=MovieSceneScriptingBoolChannel
        default=false
        keys [[0,false],[1,false],[2,false],[3,false],[4,false],[5,false],[6,false],[7,false],[8,false],[9,false],[10,false],[11,false],[12,false],[13,false],[14,false],[15,false],[16,false],[17,false],[18,false],[19,false],[20,false],[21,false],[22,false],[23,false],[24,false],[25,false],[26,false],[27,false],[28,false],[29,false],[30,false],[31,false],[32,false],[33,false],[34,false],[35,false],[36,false],[37,false],[38,false],[39,false],[40,false],[41,false],[42,false],[43,false],[44,false],[45,false],[46,false],[47,false],[48,false],[49,false],[50,false],[51,false],[52,false],[53,false],[54,false],[55,false],[56,false],[57,false],[58,false],[59,false],[60,false],[61,false],[62,false],[63,false],[64,false],[65,false],[66,false],[67,false],[68,false],[69,false],[70,false],[71,false],[72,false],[73,false],[74,false],[75,false],[76,false],[77,false],[78,false],[79,false],[80,false],[81,false],[82,false],[83,false],[84,false],[85,false],[86,false],[87,false],[88,false],[89,false],[90,false],[91,false],[92,false],[93,false],[94,false],[95,false],[96,false],[97,false],[98,false],[99,false],[100,false],[101,false],[102,false],[103,false],[104,false],[105,false],[106,false],[107,false],[108,false],[109,false],[110,false],[111,false],[112,false],[113,false],[114,false],[115,false],[116,false],[117,false],[118,false],[119,false],[120,false],[121,false],[122,false],[123,false],[124,false],[125,false],[126,false],[127,false],[128,false],[129,false],[130,false],[131,false],[132,false],[133,false],[134,false],[135,false],[136,false],[137,false],[138,false],[139,false],[140,false],[141,false],[142,false],[143,false],[144,false],[145,false],[146,false],[147,false],[148,false],[149,false],[150,false],[151,false],[152,false],[153,false],[154,false],[155,false],[156,false],[157,false],[158,false],[159,false],[160,false],[161,false],[162,false],[163,false],[164,false],[165,false],[166,false],[167,false],[168,false],[169,false],[170,false],[171,false],[172,false],[173,true],[174,false],[175,true],[176,false],[177,false],[178,true],[179,false],[180,true],[181,true],[182,true],[183,false],[184,false]]
       }
       head_local {type=MovieSceneScriptingBoolChannel, default=true, const [true,182]}
       Stretch {
        type=MovieSceneScriptingBoolChannel
        default=false
        keys [[0,false],[1,false],[2,false],[3,false],[4,false],[5,false],[6,false],[7,false],[8,false],[9,false],[10,false],[11,false],[12,false],[13,false],[14,false],[15,false],[16,false],[17,false],[18,false],[19,false],[20,false],[21,false],[22,false],[23,false],[24,false],[25,false],[26,false],[27,false],[28,false],[29,false],[30,false],[31,false],[32,false],[33,false],[34,false],[35,false],[36,false],[37,false],[38,false],[39,false],[40,false],[41,false],[42,false],[43,false],[44,false],[45,false],[46,false],[47,false],[48,false],[49,false],[50,false],[51,false],[52,false],[53,false],[54,false],[55,false],[56,false],[57,false],[58,false],[59,false],[60,false],[61,false],[62,false],[63,false],[64,false],[65,false],[66,false],[67,false],[68,false],[69,false],[70,false],[71,false],[72,false],[73,false],[74,false],[75,false],[76,false],[77,false],[78,false],[79,false],[80,false],[81,false],[82,false],[83,false],[84,false],[85,false],[86,false],[87,false],[88,false],[89,false],[90,false],[91,false],[92,false],[93,false],[94,false],[95,false],[96,false],[97,false],[98,false],[99,false],[100,false],[101,false],[102,false],[103,false],[106,false],[107,false],[108,false],[109,false],[110,false],[111,false],[112,false],[113,false],[114,false],[115,false],[116,false],[117,false],[118,false],[119,false],[120,false],[121,false],[122,false],[123,false],[124,false],[125,false],[126,false],[127,false],[128,false],[129,false],[130,false],[131,false],[132,false],[133,false],[134,false],[135,false],[136,false],[137,false],[138,false],[139,false],[140,false],[141,false],[142,false],[143,false],[144,false],[145,false],[146,false],[147,false],[148,false],[149,false],[150,false],[151,false],[152,false],[153,false],[154,false],[155,false],[156,false],[157,false],[158,false],[159,false],[160,false],[161,false],[162,false],[163,false],[164,false],[165,false],[166,false],[167,false],[168,false],[169,false],[170,false],[171,false],[172,false],[173,true],[174,false],[175,true],[176,false],[177,false],[178,true],[179,false],[180,true],[181,true],[182,true],[184,false]]
       }
       head_stretch {
        type=MovieSceneScriptingBoolChannel
        default=false
        keys [[0,false],[1,false],[2,false],[3,false],[4,false],[5,false],[6,false],[7,false],[8,false],[9,false],[10,false],[11,false],[12,false],[13,false],[14,false],[15,false],[16,false],[17,false],[18,false],[19,false],[20,false],[21,false],[22,false],[23,false],[24,false],[25,false],[26,false],[27,false],[28,false],[29,false],[30,false],[31,false],[32,false],[33,false],[34,false],[35,false],[36,false],[37,false],[38,false],[39,false],[40,false],[41,false],[42,false],[43,false],[44,false],[45,false],[46,false],[47,false],[48,false],[49,false],[50,false],[51,false],[52,false],[53,false],[54,false],[55,false],[56,false],[57,false],[58,false],[59,false],[60,false],[61,false],[62,false],[63,false],[64,false],[65,false],[66,false],[67,false],[68,false],[69,false],[70,false],[71,false],[72,false],[73,false],[74,false],[75,false],[76,false],[77,false],[78,false],[79,false],[80,false],[81,false],[82,false],[83,false],[84,false],[85,false],[86,false],[87,false],[88,false],[89,false],[90,false],[91,false],[92,false],[93,false],[94,false],[95,false],[96,false],[97,false],[98,false],[99,false],[100,false],[101,false],[102,false],[103,false],[104,false],[105,false],[106,false],[107,false],[108,false],[109,false],[110,false],[111,false],[112,false],[113,false],[114,false],[115,false],[116,false],[117,false],[118,false],[119,false],[120,false],[121,false],[122,false],[123,false],[124,false],[125,false],[126,false],[127,false],[128,false],[129,false],[130,false],[131,false],[132,false],[133,false],[134,false],[135,false],[136,false],[137,false],[138,false],[139,false],[140,false],[141,false],[142,false],[143,false],[144,false],[145,false],[146,false],[147,false],[148,false],[149,false],[150,false],[151,false],[152,false],[153,false],[154,false],[155,false],[156,false],[157,false],[158,false],[159,false],[160,false],[161,false],[162,false],[163,false],[164,false],[165,false],[166,false],[167,false],[168,false],[169,false],[170,false],[171,false],[172,false],[173,true],[174,false],[175,true],[176,false],[177,false],[178,true],[179,false],[180,true],[181,true],[182,true],[183,false],[184,false]]
       }
       spine_stretch_switch {
        type=MovieSceneScriptingBoolChannel
        default=false
        keys [[0,false],[1,false],[2,false],[3,false],[4,false],[5,false],[6,false],[7,false],[8,false],[9,false],[10,false],[11,false],[12,false],[13,false],[14,false],[15,false],[16,false],[17,false],[18,false],[19,false],[20,false],[21,false],[22,false],[23,false],[24,false],[25,false],[26,false],[27,false],[28,false],[29,false],[30,false],[31,false],[32,false],[33,false],[34,false],[35,false],[36,false],[37,false],[38,false],[39,false],[40,false],[41,false],[42,false],[43,false],[44,false],[45,false],[46,false],[47,false],[48,false],[49,false],[50,false],[51,false],[52,false],[53,false],[54,false],[55,false],[56,false],[57,false],[58,false],[59,false],[60,false],[61,false],[62,false],[63,false],[64,false],[65,false],[66,false],[67,false],[68,false],[69,false],[70,false],[71,false],[72,false],[73,false],[74,false],[75,false],[76,false],[77,false],[78,false],[79,false],[80,false],[81,false],[82,false],[83,false],[84,false],[85,false],[86,false],[87,false],[88,false],[89,false],[90,false],[91,false],[92,false],[93,false],[94,false],[95,false],[96,false],[97,false],[98,false],[99,false],[100,false],[101,false],[102,false],[103,false],[104,false],[105,false],[106,false],[107,false],[108,false],[109,false],[110,false],[111,false],[112,false],[113,false],[114,false],[115,false],[116,false],[117,false],[118,false],[119,false],[120,false],[121,false],[122,false],[123,false],[124,false],[125,false],[126,false],[127,false],[128,false],[129,false],[130,false],[131,false],[132,false],[133,false],[134,false],[135,false],[136,false],[137,false],[138,false],[139,false],[140,false],[141,false],[142,false],[143,false],[144,false],[145,false],[146,false],[147,false],[148,false],[149,false],[150,false],[151,false],[152,false],[153,false],[154,false],[155,false],[156,false],[157,false],[158,false],[159,false],[160,false],[161,false],[162,false],[163,false],[164,false],[165,false],[166,false],[167,false],[168,false],[169,false],[170,false],[171,false],[172,false],[173,true],[174,false],[175,true],[176,false],[177,false],[178,true],[179,false],[180,true],[181,true],[182,true],[183,false],[184,false]]
       }
       arm_l_fk_ik_switch {type=MovieSceneScriptingBoolChannel, default=false, const [false,179]}
       leg_l_fk_ik_switch {type=MovieSceneScriptingBoolChannel, default=false, const [false,179]}
       arm_r_fk_ik_switch {type=MovieSceneScriptingBoolChannel, default=false, const [false,179]}
       leg_r_fk_ik_switch {type=MovieSceneScriptingBoolChannel, default=false, const [false,179]}
       ShowBodyControls {type=MovieSceneScriptingBoolChannel, default=true, const [true,185]}
       spine_fk_ik_switch {type=MovieSceneScriptingBoolChannel, default=false, const [false,180]}
       neck_fk_ik_switch {type=MovieSceneScriptingBoolChannel, default=false, const [false,180]}
       space {type=MovieSceneScriptingByteChannel, default=1, const [1,180]}
       global_ctrl.Location.X {type=MovieSceneScriptingFloatChannel, default=0, const [0,185]}
       global_ctrl.Location.Y {type=MovieSceneScriptingFloatChannel, default=0, const [0,185]}
       global_ctrl.Location.Z {type=MovieSceneScriptingFloatChannel, default=0, const [0,185]}
       global_ctrl.Rotation.X {type=MovieSceneScriptingFloatChannel, default=0, const [0,185]}
       global_ctrl.Rotation.Y {type=MovieSceneScriptingFloatChannel, default=0, const [0,185]}
       global_ctrl.Rotation.Z {type=MovieSceneScriptingFloatChannel, default=0, const [0,185]}
       root_ctrl.Location.X {type=MovieSceneScriptingFloatChannel, default=0, const [0,183]}
       root_ctrl.Location.Y {type=MovieSceneScriptingFloatChannel, default=0, const [0,183]}
       root_ctrl.Location.Z {type=MovieSceneScriptingFloatChannel, default=0, const [0,183]}
       root_ctrl.Rotation.X {type=MovieSceneScriptingFloatChannel, default=0, const [0,183]}
       root_ctrl.Rotation.Y {type=MovieSceneScriptingFloatChannel, default=0, const [0,183]}
       root_ctrl.Rotation.Z {type=MovieSceneScriptingFloatChannel, default=0, const [0,183]}
       root_ctrl.Scale.X {type=MovieSceneScriptingFloatChannel, default=1, const [1,183]}
       root_ctrl.Scale.Y {type=MovieSceneScriptingFloatChannel, default=1, const [1,183]}
       root_ctrl.Scale.Z {type=MovieSceneScriptingFloatChannel, default=1, const [1,183]}
       body_offset_ctrl.Location.X {type=MovieSceneScriptingFloatChannel, default=0, const [0,183]}
       body_offset_ctrl.Location.Y {type=MovieSceneScriptingFloatChannel, default=0, const [0,183]}
       body_offset_ctrl.Location.Z {type=MovieSceneScriptingFloatChannel, default=0, const [0,183]}
       body_offset_ctrl.Rotation.X {type=MovieSceneScriptingFloatChannel, default=0, const [0,183]}
       body_offset_ctrl.Rotation.Y {type=MovieSceneScriptingFloatChannel, default=0, const [0,183]}
       body_offset_ctrl.Rotation.Z {type=MovieSceneScriptingFloatChannel, default=0, const [0,183]}
       body_offset_ctrl.Scale.X {type=MovieSceneScriptingFloatChannel, default=1, const [1,183]}
       body_offset_ctrl.Scale.Y {type=MovieSceneScriptingFloatChannel, default=1, const [1,183]}
       body_offset_ctrl.Scale.Z {type=MovieSceneScriptingFloatChannel, default=1, const [1,183]}
       thigh_l_fk_ctrl.Location.X {type=MovieSceneScriptingFloatChannel, default=0, thinned_from=173, keys [[0,0],[184,0]]}
       thigh_l_fk_ctrl.Location.Y {type=MovieSceneScriptingFloatChannel, default=0, thinned_from=173, keys [[0,0],[184,0]]}
       thigh_l_fk_ctrl.Location.Z {type=MovieSceneScriptingFloatChannel, default=0, thinned_from=173, keys [[0,0],[184,0]]}
       thigh_l_fk_ctrl.Rotation.X {
        type=MovieSceneScriptingFloatChannel
        default=11.767
        thinned_from=173
        keys [[0,1.5753],[1,1.8938],[2,2.0538],[3,2.0126],[4,1.8844],[5,1.6896],[6,1.6728],[7,1.9552],[8,2.2753],[9,2.3108],[10,2.2222],[11,2.0557],[12,1.8981],[13,1.894],[14,2.1626],[15,2.2979],[16,2.5286],[17,2.6887],[18,2.829],[19,3.1062],[20,3.247],[21,3.4387],[22,3.6814],[23,3.8895],[24,3.938],[25,3.9157],[26,3.6919],[27,3.4548],[28,3.125],[29,2.7154],[30,2.2088],[31,1.5386],[32,1.1158],[33,0.9135],[34,0.8738],[35,0.8536],[36,1.0539],[37,1.1985],[38,1.2575],[39,1.3787],[40,1.7239],[41,2.0122],[42,2.3281],[43,2.6093],[45,2.7838],[46,3.3099],[47,3.9297],[48,4.6804],[49,5.3981],[50,5.9512],[51,6.2255],[52,6.555],[53,6.5292],[54,6.1847],[55,4.7637],[56,4.1019],[57,3.209],[58,2.8886],[59,2.9488],[60,2.9499],[61,2.1926],[62,2.0121],[63,1.3593],[64,1.0371],[65,1.2137],[66,1.8529],[67,2.7561],[68,3.291],[69,4.1703],[70,5.5861],[71,6.4431],[72,7.8949],[73,8.9799],[74,9.7374],[75,10.6813],[76,11.4843],[77,11.9542],[78,12.2315],[80,12.6678],[81,12.7811],[82,13.0404],[83,13.4273],[86,14.0057],[87,13.8524],[88,13.6223],[89,12.9824],[90,12.41],[91,12.128],[92,12.116],[95,12.4375],[96,12.7773],[97,13.3381],[98,13.6885],[99,13.6792],[100,13.7234],[102,13.6773],[103,13.5799],[106,13.347],[107,13.1912],[108,13.0621],[110,13.4422],[111,13.5618],[112,13.6014],[113,13.8134],[114,14.1486],[115,14.4126],[116,14.5508],[117,14.6347],[119,15.0048],[120,15.0949],[121,15.1421],[122,15.2286],[123,15.4636],[124,15.5923],[125,15.6422],[127,15.8272],[128,15.9892],[129,15.9957],[132,15.7396],[134,15.7521],[136,15.505],[137,15.5187],[138,15.5833],[139,15.6919],[140,15.8433],[141,15.8677],[142,15.7604],[143,15.6181],[145,15.4985],[146,15.3547],[147,15.2553],[148,15.2428],[149,15.1606],[150,15.1391],[151,15.2148],[152,15.3376],[153,15.6174],[154,15.9891],[156,16.4748],[157,16.5978],[158,16.8279],[159,17.0949],[160,17.4542],[161,17.5264],[162,17.7846],[163,17.9313],[164,17.8002],[165,17.3917],[166,17.2056],[167,17.1741],[168,16.9121],[169,17.7465],[170,17.7738],[171,20.0306],[172,22.197],[181,24.2225],[184,0]]
       }
       thigh_l_fk_ctrl.Rotation.Y {
        type=MovieSceneScriptingFloatChannel
        default=-4.485
        thinned_from=173
        keys [[0,-3.9375],[1,-3.8681],[2,-3.8173],[4,-3.9872],[5,-4.0241],[6,-3.9146],[8,-3.4658],[9,-3.4565],[11,-3.3991],[13,-3.321],[14,-3.2417],[15,-3.2317],[16,-3.1971],[17,-3.1842],[18,-3.1641],[19,-3.0595],[20,-3.0142],[21,-3.0412],[22,-2.9623],[23,-2.8921],[24,-2.9669],[25,-3.1514],[26,-3.3735],[27,-3.4365],[28,-3.5631],[29,-3.844],[30,-4.1487],[31,-4.4172],[32,-4.4552],[33,-4.3437],[34,-4.3483],[35,-4.4085],[36,-4.2205],[37,-4.0506],[38,-4.0319],[39,-3.9588],[40,-3.6757],[41,-3.5368],[42,-3.4537],[43,-3.2828],[44,-3.1749],[45,-3.0974],[46,-2.686],[47,-2.359],[48,-2.0828],[49,-1.9993],[50,-2.0078],[51,-2.0426],[52,-2.1475],[53,-2.9539],[54,-3.7258],[55,-5.1826],[56,-6.0085],[57,-7.0759],[58,-7.8818],[59,-8.546],[61,-9.4987],[62,-9.6057],[63,-10.1213],[64,-10.6937],[65,-10.8161],[66,-10.9642],[67,-10.8869],[68,-11.0201],[69,-11.1864],[70,-11.1282],[71,-10.7948],[72,-10.1922],[73,-9.9515],[74,-9.8348],[76,-9.3403],[78,-9.0988],[79,-8.9967],[80,-8.935],[82,-8.9488],[83,-8.9378],[84,-8.8893],[85,-8.9407],[86,-9.1849],[87,-9.3803],[88,-9.4122],[89,-9.5136],[91,-9.8236],[92,-9.9284],[95,-9.9385],[96,-9.8878],[98,-9.5883],[100,-9.4597],[101,-9.456],[102,-9.5208],[103,-9.6234],[106,-9.5798],[107,-9.5825],[110,-9.4402],[111,-9.3257],[113,-9.1872],[114,-9.151],[115,-9.1426],[116,-9.0242],[117,-8.9478],[118,-8.9118],[121,-8.8741],[122,-8.8444],[123,-8.7875],[124,-8.7531],[125,-8.6939],[126,-8.6129],[128,-8.5531],[130,-8.5639],[131,-8.6008],[132,-8.6633],[133,-8.69],[134,-8.8314],[135,-9.0358],[136,-9.1466],[137,-9.16],[138,-9.1392],[140,-9.0063],[141,-8.9956],[142,-9.0503],[143,-9.1478],[144,-9.2655],[146,-9.5375],[147,-9.6172],[149,-9.644],[150,-9.6099],[151,-9.491],[152,-9.3312],[153,-9.094],[154,-8.752],[156,-8.2902],[157,-8.1271],[158,-7.9462],[159,-7.8577],[160,-7.7014],[161,-7.6514],[162,-7.5125],[163,-7.4741],[165,-7.5771],[166,-7.3497],[167,-6.5855],[168,-2.8577],[169,-0.1554],[170,1.3701],[171,2.927],[172,4.0453],[181,1.9234],[184,-7.0502]]
       }
       thigh_l_fk_ctrl.Rotation.Z {
        type=MovieSceneScriptingFloatChannel
        default=-1.0628
        thinned_from=173
        keys [[0,8.5455],[1,8.4776],[2,8.346],[3,8.2405],[4,8.1672],[5,8.0493],[6,7.9664],[7,7.9985],[9,8.0077],[11,7.979],[12,7.9474],[13,7.9415],[14,7.9805],[15,7.9987],[18,8.1538],[19,8.2571],[20,8.295],[21,8.3495],[22,8.4321],[23,8.4695],[24,8.4892],[25,8.4773],[26,8.4044],[27,8.3538],[28,8.3518],[29,8.374],[30,8.3549],[32,8.2288],[33,8.1904],[34,8.1728],[35,8.2251],[36,8.2295],[38,8.1613],[39,8.1431],[40,8.1472],[41,8.1197],[42,8.0328],[43,7.8704],[44,7.7639],[45,7.7365],[46,7.8316],[47,8.0673],[48,8.5912],[49,9.3761],[50,10.5225],[51,10.9883],[52,11.7727],[53,13.989],[54,14.5923],[55,16.7934],[56,17.4636],[57,18.6069],[58,19.4986],[59,20.8661],[61,23.4785],[62,23.0764],[63,24.0191],[64,24.6388],[65,23.9724],[66,23.4071],[67,21.9702],[68,20.6646],[69,19.565],[70,18.6002],[71,17.5389],[72,17.1536],[73,17.0175],[74,16.4828],[75,15.6861],[76,14.9997],[78,13.7558],[79,13.1593],[80,12.8134],[82,12.5175],[83,12.3201],[84,11.977],[85,11.7249],[86,11.6028],[88,11.0195],[90,10.1596],[91,9.7676],[92,9.3466],[93,9.0555],[94,8.7086],[95,8.5258],[96,8.3946],[97,8.324],[98,8.3069],[99,8.3389],[100,8.3429],[101,8.2869],[102,8.2533],[106,8.2548],[107,8.2034],[108,8.1348],[111,8.3078],[112,8.3346],[114,8.534],[115,8.6982],[116,8.7114],[118,8.9828],[119,9.0741],[120,9.1323],[121,9.1474],[123,9.1237],[124,9.1315],[125,9.1056],[126,9.0425],[127,9.0204],[128,9.0691],[129,9.093],[130,9.0868],[132,9.1515],[134,9.1594],[136,9.3432],[137,9.3847],[138,9.4026],[139,9.3734],[140,9.2605],[142,9.0033],[143,8.9964],[144,9.015],[146,9.1673],[148,9.1314],[150,9.0011],[152,8.6978],[153,8.5886],[154,8.5025],[155,8.523],[156,8.5668],[157,8.6542],[158,8.7645],[159,8.9051],[160,9.0023],[161,9.0185],[162,8.9657],[163,8.8932],[164,8.7971],[165,8.6328],[166,8.1032],[167,6.1979],[168,-1.1435],[169,-7.5179],[170,-12.6506],[171,-14.8702],[172,-13.7678],[181,-12.038],[184,0]]
       }
       thigh_l_fk_ctrl.Scale.X {type=MovieSceneScriptingFloatChannel, default=1, const [1,173]}
       thigh_l_fk_ctrl.Scale.Y {type=MovieSceneScriptingFloatChannel, default=1, const [1,173]}
       thigh_l_fk_ctrl.Scale.Z {type=MovieSceneScriptingFloatChannel, default=1, const [1,173]}
       calf_l_fk_ctrl.Location.X {
        type=MovieSceneScriptingFloatChannel
        default=0.0008
        thinned_from=173
        keys [[0,0.0001],[1,0.0005],[2,-0.0001],[3,0.0004],[5,0.0007],[6,-0.0001],[8,0],[9,0.0002],[10,0.0001],[11,0.0003],[12,-0.0002],[14,0.0001],[17,-0.0001],[19,0.0002],[20,-0.0001],[22,0.0002],[24,0],[26,0.0005],[28,-0.0001],[30,0.0008],[31,0.0009],[33,0.0005],[34,0.0001],[35,0.0003],[36,0.0002],[37,0.0005],[38,-0.0002],[39,0],[40,0.0005],[41,0.0005],[43,0.0002],[44,0.0004],[45,0.0001],[46,0.0003],[47,0.0001],[52,0.0002],[54,0.0008],[56,0.0021],[57,0.0017],[58,0.002],[59,0.0016],[60,0.0002],[61,0.0002],[62,-0.0001],[63,0.0007],[64,0],[65,0.0001],[159,-0.0001],[165,-0.0001],[166,0.0006],[167,0],[168,0.0028],[169,0.0018],[170,0.0024],[172,0.0007],[181,0],[184,0]]
       }
       calf_l_fk_ctrl.Location.Y {
        type=MovieSceneScriptingFloatChannel
        default=0.0089
        thinned_from=173
        keys [[0,0.0017],[1,0.0081],[2,-0.0017],[3,0.0057],[4,0.0083],[5,0.0115],[6,-0.0017],[7,0.0004],[8,0.0002],[9,0.0041],[10,0.0015],[11,0.0043],[12,-0.0024],[13,-0.0016],[14,0.0016],[15,-0.0003],[17,-0.0013],[19,0.0032],[20,-0.0013],[21,0.0003],[22,0.003],[23,0.0008],[24,0.0001],[25,0.0031],[26,0.0073],[27,0.0037],[28,-0.0018],[29,0.0058],[30,0.012],[31,0.0137],[32,0.0104],[33,0.0079],[34,0.0019],[35,0.0053],[36,0.0031],[37,0.0075],[38,-0.0028],[39,-0.0004],[40,0.0086],[41,0.0084],[42,0.0054],[43,0.0034],[44,0.0069],[45,0.0014],[46,0.0041],[47,0.0018],[48,0.0024],[49,0.0024],[50,0.0022],[51,0.0024],[52,0.0001],[53,0.0003],[54,0.0021],[55,0.0065],[56,0.019],[57,0.0006],[58,0.0194],[59,-0.0019],[60,-0.0012],[61,-0.0008],[62,-0.0012],[63,-0.0002],[64,-0.0002],[65,0.0007],[66,-0.0003],[69,-0.0002],[70,0.0003],[71,-0.0003],[72,-0.0004],[73,-0.0001],[74,-0.0004],[76,0.0002],[77,-0.0002],[79,0.0004],[81,0.0004],[83,-0.0005],[84,0.0001],[85,-0.0001],[87,0.0001],[88,0.0004],[90,-0.0001],[93,0],[94,0.0002],[95,-0.0002],[96,0],[97,0],[98,-0.0004],[99,0.0002],[101,-0.0001],[102,0.0004],[103,-0.0002],[110,0.0001],[111,0.0003],[113,-0.0003],[116,-0.0002],[117,0.0002],[118,-0.0003],[119,-0.0004],[120,0.0003],[123,-0.0005],[126,0.0002],[127,-0.0001],[129,0],[130,-0.0004],[131,0],[132,-0.0005],[134,0.0003],[136,-0.0002],[139,-0.0002],[141,0.0004],[142,-0.0001],[144,0.0001],[145,-0.0004],[147,0.0003],[148,-0.0003],[149,0.0003],[150,0.0001],[152,0.0002],[153,-0.0001],[154,0.0001],[156,-0.0001],[157,-0.0005],[159,-0.0006],[160,0.0005],[162,0.0002],[163,0.0005],[164,-0.0006],[165,-0.0002],[166,-0.0002],[167,0.0005],[168,0.0209],[169,0.003],[170,0.0025],[171,0.0047],[172,0.0078],[181,0],[184,0]]
       }
       calf_l_fk_ctrl.Location.Z {
        type=MovieSceneScriptingFloatChannel
        default=0.0022
        thinned_from=173
        keys [[0,0.0009],[1,0.0043],[2,-0.0009],[3,0.0031],[4,0.0043],[5,0.0059],[6,-0.0009],[7,0.0002],[8,0.0001],[9,0.0024],[10,0.0009],[11,0.0025],[12,-0.0014],[13,-0.0009],[14,0.001],[15,-0.0002],[17,-0.0008],[19,0.002],[20,-0.0008],[21,0.0002],[22,0.002],[23,0.0006],[24,0.0001],[25,0.002],[26,0.0045],[27,0.0023],[28,-0.0011],[29,0.0032],[30,0.0061],[31,0.0065],[32,0.0047],[33,0.0037],[34,0.0009],[35,0.0025],[36,0.0015],[37,0.0038],[38,-0.0014],[39,-0.0002],[40,0.0047],[41,0.0047],[42,0.0031],[43,0.0021],[44,0.0043],[45,0.0009],[46,0.0029],[47,0.0013],[48,0.0019],[51,0.0018],[52,0.0001],[53,0],[54,0.0008],[55,0.0019],[56,0.0055],[57,-0.0003],[58,0.0039],[59,-0.0004],[62,-0.0002],[63,-0.0005],[64,0],[81,0.0001],[83,-0.0001],[84,0],[98,-0.0001],[102,0.0001],[103,0],[117,0],[119,-0.0001],[120,0.0001],[124,-0.0001],[131,0],[132,-0.0001],[134,0.0001],[141,0.0001],[145,-0.0001],[147,0.0001],[148,-0.0001],[152,0],[159,-0.0002],[160,0.0002],[162,0.0001],[163,0.0002],[164,-0.0002],[167,0.0002],[168,0.0069],[169,0.0013],[170,0.0014],[172,0.0025],[181,0],[184,0]]
       }
       calf_l_fk_ctrl.Rotation.X {
        type=MovieSceneScriptingFloatChannel
        default=-1.8303
        thinned_from=173
        keys [[0,-4.0186],[1,-4.0265],[2,-4.0223],[5,-4.0516],[10,-4.0269],[14,-4.0453],[22,-3.9981],[24,-3.9797],[27,-3.9957],[28,-3.9874],[34,-3.9938],[37,-4.012],[38,-4.0077],[40,-4.027],[43,-4.0324],[45,-4.0412],[47,-4.0726],[49,-4.0849],[51,-4.0586],[52,-4.0291],[53,-3.9621],[54,-3.9038],[55,-3.755],[56,-3.6438],[57,-3.4697],[58,-3.2829],[59,-3.039],[61,-2.5935],[62,-2.4944],[63,-2.3113],[64,-2.1601],[65,-2.1001],[66,-2.0489],[67,-2.0789],[68,-2.1019],[69,-2.0944],[70,-2.0828],[71,-2.0815],[72,-2.0341],[73,-1.9454],[74,-1.8814],[75,-1.8456],[77,-1.7621],[78,-1.727],[79,-1.6967],[80,-1.6472],[81,-1.5899],[82,-1.5436],[83,-1.5142],[84,-1.5063],[85,-1.4853],[86,-1.4423],[87,-1.4098],[88,-1.3907],[89,-1.3823],[91,-1.3735],[92,-1.3784],[93,-1.3772],[94,-1.3851],[96,-1.3766],[99,-1.3494],[100,-1.3469],[101,-1.3524],[102,-1.3521],[103,-1.346],[106,-1.3465],[107,-1.3551],[108,-1.3712],[110,-1.3788],[112,-1.3891],[114,-1.3788],[115,-1.3652],[116,-1.372],[117,-1.3666],[119,-1.3449],[120,-1.3407],[121,-1.3396],[124,-1.3495],[125,-1.3573],[126,-1.373],[128,-1.3874],[130,-1.3876],[132,-1.3726],[133,-1.3692],[134,-1.3571],[136,-1.2951],[137,-1.2798],[138,-1.2732],[139,-1.2774],[140,-1.29],[141,-1.299],[142,-1.3018],[143,-1.295],[144,-1.285],[146,-1.2405],[147,-1.227],[149,-1.2195],[150,-1.2231],[151,-1.2454],[153,-1.3072],[154,-1.347],[155,-1.3673],[156,-1.3844],[157,-1.3895],[158,-1.3859],[159,-1.3726],[160,-1.3628],[162,-1.3544],[164,-1.3404],[165,-1.3436],[166,-1.3861],[167,-1.5642],[168,-2.1855],[169,-2.7919],[170,-3.3454],[171,-3.6834],[172,-3.8785],[181,0.9789],[184,-1.7662]]
       }
       calf_l_fk_ctrl.Rotation.Y {
        type=MovieSceneScriptingFloatChannel
        default=-0.6531
        thinned_from=173
        keys [[0,-0.5168],[1,-0.5121],[2,-0.5008],[4,-0.4954],[5,-0.4915],[6,-0.4825],[8,-0.493],[9,-0.5018],[10,-0.5004],[11,-0.4957],[12,-0.4802],[14,-0.4626],[15,-0.4605],[17,-0.4671],[18,-0.4696],[19,-0.4708],[20,-0.4671],[22,-0.4656],[23,-0.4594],[24,-0.4593],[25,-0.465],[26,-0.4721],[27,-0.4762],[28,-0.4917],[29,-0.5178],[30,-0.5358],[31,-0.5435],[32,-0.5442],[33,-0.5397],[34,-0.5371],[35,-0.5398],[37,-0.5218],[38,-0.5067],[40,-0.4839],[42,-0.4575],[44,-0.4087],[45,-0.3907],[47,-0.3105],[48,-0.2681],[49,-0.2508],[50,-0.2575],[51,-0.2358],[52,-0.19],[53,-0.2121],[54,-0.1643],[55,-0.2181],[56,-0.1688],[57,-0.1497],[58,-0.1316],[59,-0.1389],[60,-0.1548],[61,-0.2025],[62,-0.2122],[63,-0.3329],[64,-0.4287],[65,-0.4482],[66,-0.4639],[67,-0.4218],[68,-0.3851],[69,-0.3686],[70,-0.3553],[71,-0.3291],[72,-0.332],[73,-0.3613],[74,-0.3719],[75,-0.3646],[76,-0.3585],[77,-0.3563],[78,-0.3507],[79,-0.3425],[80,-0.3439],[81,-0.3484],[82,-0.3471],[83,-0.3399],[84,-0.3242],[85,-0.3146],[86,-0.3152],[87,-0.3119],[88,-0.3045],[90,-0.282],[91,-0.2718],[92,-0.2589],[93,-0.2491],[94,-0.2365],[95,-0.2298],[96,-0.2256],[97,-0.2238],[98,-0.2235],[99,-0.2251],[100,-0.2253],[101,-0.2227],[102,-0.2231],[103,-0.2254],[106,-0.2246],[107,-0.2207],[108,-0.2139],[110,-0.2123],[112,-0.2081],[113,-0.2102],[114,-0.2133],[115,-0.2181],[116,-0.214],[117,-0.2154],[119,-0.2241],[120,-0.2264],[121,-0.2268],[122,-0.2257],[124,-0.221],[125,-0.2168],[126,-0.21],[128,-0.203],[129,-0.203],[130,-0.2025],[131,-0.2053],[132,-0.2091],[133,-0.2108],[134,-0.2165],[135,-0.2295],[136,-0.2414],[137,-0.2469],[138,-0.2486],[139,-0.2462],[140,-0.241],[141,-0.2371],[142,-0.2353],[143,-0.2382],[144,-0.2426],[146,-0.2599],[147,-0.2649],[148,-0.2658],[149,-0.2675],[150,-0.266],[151,-0.2573],[152,-0.2453],[153,-0.2302],[154,-0.2095],[155,-0.1981],[156,-0.1884],[157,-0.1847],[158,-0.1844],[159,-0.1887],[160,-0.1909],[161,-0.1913],[162,-0.1899],[163,-0.192],[164,-0.1965],[165,-0.1973],[166,-0.1762],[167,-0.0654],[168,0.4725],[169,1.045],[170,1.6921],[171,2.2958],[172,2.6377],[181,0.7365],[184,-1.1341]]
       }
       calf_l_fk_ctrl.Rotation.Z {
        type=MovieSceneScriptingFloatChannel
        default=-6.8131
        thinned_from=173
        keys [[0,3.8897],[2,4.0916],[3,4.1611],[4,4.2004],[6,4.3031],[7,4.2401],[8,4.1536],[9,4.0512],[10,4.0488],[11,4.1239],[12,4.2976],[13,4.4281],[14,4.57],[15,4.5941],[17,4.5142],[18,4.5012],[19,4.5091],[22,4.6107],[23,4.6907],[24,4.6955],[25,4.638],[26,4.5588],[27,4.4681],[28,4.2106],[29,3.8824],[30,3.6481],[31,3.5278],[32,3.4863],[33,3.5287],[34,3.5374],[35,3.5172],[36,3.6276],[37,3.7754],[39,4.117],[40,4.3262],[42,4.6854],[43,5.0075],[44,5.3655],[45,5.5814],[46,6.1515],[47,6.7041],[48,7.324],[49,7.6163],[50,7.6111],[51,8.0694],[52,9.0809],[53,9.4435],[54,10.7189],[55,10.5804],[56,11.7402],[57,12.1976],[58,12.5013],[59,12.0892],[60,11.4371],[61,10.1092],[62,9.5451],[63,7.3984],[64,5.795],[65,5.2427],[66,4.7979],[67,5.2965],[68,5.6981],[69,5.7211],[70,5.6719],[71,5.7466],[72,5.3694],[73,4.5125],[74,3.8774],[75,3.5373],[76,3.1837],[77,2.7125],[78,2.2816],[79,1.9021],[80,1.3181],[81,0.66],[82,0.1449],[83,-0.2261],[84,-0.35],[85,-0.615],[86,-1.1262],[87,-1.5596],[88,-1.8611],[89,-2.0521],[90,-2.1956],[91,-2.3248],[92,-2.3325],[93,-2.3969],[94,-2.3542],[95,-2.4198],[96,-2.5107],[99,-2.8622],[100,-2.9136],[101,-2.8494],[102,-2.8602],[103,-2.919],[106,-2.9243],[107,-2.831],[108,-2.6642],[109,-2.6484],[110,-2.6189],[112,-2.5114],[113,-2.5446],[114,-2.594],[115,-2.6698],[116,-2.5507],[117,-2.5648],[119,-2.7732],[120,-2.8313],[121,-2.8413],[122,-2.8101],[124,-2.6574],[125,-2.5327],[126,-2.3511],[127,-2.2536],[128,-2.1247],[129,-2.1091],[130,-2.1024],[131,-2.1708],[132,-2.2738],[133,-2.3185],[134,-2.4842],[136,-3.2242],[137,-3.4022],[138,-3.4584],[139,-3.3883],[140,-3.2393],[141,-3.1395],[142,-3.1081],[143,-3.1976],[144,-3.3317],[145,-3.6092],[146,-3.8743],[147,-4.0477],[148,-4.0882],[149,-4.1508],[150,-4.1112],[151,-3.8325],[152,-3.46],[153,-2.9842],[154,-2.3587],[155,-2.0233],[156,-1.7441],[157,-1.632],[158,-1.5997],[159,-1.6786],[160,-1.6996],[161,-1.7036],[162,-1.6807],[163,-1.7521],[164,-1.9315],[165,-2.0145],[166,-1.6503],[167,0.3524],[168,9.3363],[169,18.4629],[170,27.4095],[171,34.6994],[172,38.3079],[181,50.8783],[184,-21.1116]]
       }
       calf_l_fk_ctrl.Scale.X {type=MovieSceneScriptingFloatChannel, default=1, const [1,173]}
       calf_l_fk_ctrl.Scale.Y {type=MovieSceneScriptingFloatChannel, default=1, const [1,173]}
       calf_l_fk_ctrl.Scale.Z {type=MovieSceneScriptingFloatChannel, default=1, const [1,173]}
       foot_l_fk_ctrl.Location.X {
        type=MovieSceneScriptingFloatChannel
        default=0.0042
        thinned_from=127
        keys [[0,0],[54,0],[58,0],[59,0.0001],[116,0],[166,0.0001],[167,0],[168,0.0004],[170,0.0006],[171,0.0005],[172,0.0001],[184,0]]
       }
       foot_l_fk_ctrl.Location.Y {type=MovieSceneScriptingFloatChannel, default=-0.0002, thinned_from=127, keys [[0,0],[172,0],[181,0.0398],[184,0]]}
       foot_l_fk_ctrl.Location.Z {
        type=MovieSceneScriptingFloatChannel
        default=0.0004
        thinned_from=127
        keys [[0,0],[169,-0.0001],[170,-0.0004],[172,-0.0001],[181,-0.8885],[184,0]]
       }
       foot_l_fk_ctrl.Rotation.X {
        type=MovieSceneScriptingFloatChannel
        default=-4.7922
        thinned_from=127
        keys [[0,9.879],[1,9.676],[2,9.5918],[3,9.7649],[4,10.1219],[5,10.5127],[6,10.6899],[7,10.4975],[8,10.228],[9,10.3086],[10,10.6135],[11,10.9707],[12,11.2683],[13,11.3188],[14,11.0005],[15,10.7121],[16,10.3319],[18,9.6408],[19,9.2312],[20,9.0341],[22,8.4495],[23,8.2436],[24,8.2232],[25,8.3677],[26,8.7472],[27,9.1475],[28,9.6212],[29,10.1337],[30,10.7997],[31,11.688],[32,12.2355],[33,12.4558],[34,12.5016],[35,12.509],[37,12.0242],[38,11.8433],[39,11.591],[40,11.0713],[41,10.6206],[42,10.1935],[43,9.8113],[44,9.655],[45,9.561],[46,8.9984],[47,8.3317],[48,7.5328],[49,6.8689],[50,6.2379],[51,5.9943],[52,5.7263],[53,5.7549],[54,6.3722],[55,7.8245],[56,8.8315],[57,9.9147],[58,10.3625],[59,10.0668],[60,9.7147],[62,10.3954],[63,11.1588],[64,11.9892],[65,12.3279],[66,12.2556],[67,11.7189],[68,11.2367],[69,10.56],[70,9.338],[71,8.639],[72,7.3807],[73,6.8521],[74,6.819],[75,6.7012],[76,6.6178],[78,6.726],[79,6.8118],[81,6.9241],[82,6.7762],[83,6.6131],[85,7.0303],[86,7.0792],[87,7.3758],[88,7.7941],[89,8.6914],[90,9.443],[91,9.8933],[92,10.0875],[94,10.1373],[95,10.075],[96,9.6852],[97,8.9754],[98,8.3959],[99,8.1616],[100,7.7446],[101,7.5815],[102,7.4926],[103,7.4953],[104,7.4034],[105,7.3471],[106,7.3325],[107,7.4737],[108,7.6349],[109,7.4595],[110,7.2618],[111,7.1123],[112,7.0581],[113,6.8496],[114,6.5638],[115,6.4664],[116,6.476],[165,2.5982],[166,2.7445],[167,3.0002],[168,3.5297],[169,3.1993],[170,2.9129],[171,0.8284],[172,-2.1017],[181,-13.4455],[184,0]]
       }
       foot_l_fk_ctrl.Rotation.Y {
        type=MovieSceneScriptingFloatChannel
        default=2.2118
        thinned_from=127
        keys [[0,8.1918],[1,8.202],[2,8.1567],[3,8.2101],[4,8.4759],[5,8.7052],[6,8.8209],[7,8.7553],[8,8.6321],[9,8.5818],[10,8.6052],[11,8.6891],[12,8.7513],[13,8.8627],[14,8.8577],[15,8.7716],[16,8.6431],[18,8.2187],[19,8.0192],[20,7.9294],[21,7.6822],[22,7.3441],[23,7.0457],[24,6.9033],[25,7.0153],[26,7.2641],[27,7.6087],[28,7.8096],[29,7.9708],[30,8.2465],[31,8.5405],[32,8.724],[34,8.8274],[35,8.9395],[36,9.0202],[37,8.9993],[38,8.9147],[40,8.8369],[41,8.7541],[42,8.7183],[43,8.5976],[44,8.5094],[45,8.5225],[46,8.5186],[47,8.4827],[48,8.311],[49,8.165],[50,7.8481],[51,7.6611],[52,7.4474],[53,7.3618],[54,7.5772],[55,7.9432],[56,8.4343],[57,8.8299],[58,8.8867],[59,8.6916],[60,8.6749],[61,8.9927],[62,9.1309],[63,9.3109],[64,9.5419],[65,9.6346],[66,9.631],[67,9.5784],[68,9.6525],[69,9.6304],[71,9.1908],[72,8.2757],[73,7.5711],[74,7.2701],[75,7.1201],[76,6.8159],[77,6.6798],[79,6.6611],[80,6.5269],[81,6.3703],[82,6.1631],[83,6.1761],[84,6.2832],[85,6.4636],[86,6.6023],[87,6.8237],[88,7.0643],[90,7.8652],[91,8.1283],[92,8.2645],[94,8.3713],[95,8.3765],[96,8.248],[97,7.9807],[98,7.7388],[99,7.6156],[100,7.5306],[101,7.4682],[102,7.4721],[103,7.4414],[104,7.4282],[105,7.3968],[107,7.555],[108,7.6104],[109,7.6212],[110,7.5659],[111,7.4359],[112,7.2844],[113,6.9767],[114,6.7516],[115,6.3601],[116,6.2166],[165,1.2379],[166,1.4039],[167,1.4396],[168,0.5105],[169,1.6285],[170,4.2131],[171,5.3334],[172,6.3652],[181,11.0218],[184,0]]
       }
       foot_l_fk_ctrl.Rotation.Z {
        type=MovieSceneScriptingFloatChannel
        default=51.6896
        thinned_from=127
        keys [[0,33.1441],[2,33.0321],[4,33.0888],[6,33.2762],[11,33.4585],[13,33.4175],[17,33.0854],[22,32.7128],[24,32.6647],[26,32.7712],[30,33.336],[31,33.528],[33,33.712],[36,33.6612],[39,33.4205],[40,33.2334],[45,32.8709],[48,31.9195],[49,31.5038],[50,31.1541],[51,30.9365],[52,30.4721],[53,29.8041],[54,29.4257],[55,29.6642],[58,29.6195],[60,29.802],[61,30.0082],[62,30.0356],[63,30.2892],[64,30.4054],[65,30.243],[66,29.7512],[68,28.4338],[70,27.3913],[72,26.6076],[73,26.4442],[80,25.8272],[84,25.7222],[88,25.9813],[90,26.2043],[91,26.2287],[95,26.0922],[97,25.8997],[103,25.804],[104,25.6996],[105,25.6939],[106,25.8502],[109,25.7982],[112,25.5585],[114,25.3359],[116,25.2486],[165,25.0724],[166,24.8473],[167,24.117],[168,21.864],[169,18.4776],[170,15.4824],[171,13.221],[172,12.0569],[181,6.9249],[184,29.9903]]
       }
       foot_l_fk_ctrl.Scale.X {type=MovieSceneScriptingFloatChannel, default=1, const [1,127]}
       foot_l_fk_ctrl.Scale.Y {type=MovieSceneScriptingFloatChannel, default=1, const [1,127]}
       foot_l_fk_ctrl.Scale.Z {type=MovieSceneScriptingFloatChannel, default=1, const [1,127]}
       ball_l_fk_ctrl.Location.X {type=MovieSceneScriptingFloatChannel, default=0.0007, thinned_from=185, keys [[0,0],[182,0],[184,0.0007]]}
       ball_l_fk_ctrl.Location.Y {type=MovieSceneScriptingFloatChannel, default=-0.0013, thinned_from=185, keys [[0,0],[182,0],[183,-0.0002],[184,-0.0013]]}
       ball_l_fk_ctrl.Location.Z {type=MovieSceneScriptingFloatChannel, default=-0.0006, thinned_from=185, keys [[0,0],[183,0],[184,-0.0006]]}
       ball_l_fk_ctrl.Rotation.X {
        type=MovieSceneScriptingFloatChannel
        default=-0.0482
        thinned_from=185
        keys [[0,-0.1789],[2,-0.1777],[3,-0.1778],[4,-0.1765],[5,-0.1748],[6,-0.1727],[7,-0.1719],[9,-0.1711],[11,-0.169],[12,-0.1692],[13,-0.1676],[14,-0.1677],[15,-0.1685],[16,-0.1698],[17,-0.1718],[18,-0.1721],[19,-0.1735],[20,-0.1742],[21,-0.1765],[22,-0.1793],[23,-0.1778],[25,-0.1778],[26,-0.1789],[27,-0.1753],[28,-0.1753],[29,-0.1767],[31,-0.1785],[33,-0.1772],[34,-0.1773],[35,-0.1769],[36,-0.1749],[37,-0.1754],[41,-0.1753],[42,-0.1733],[43,-0.1702],[44,-0.1683],[45,-0.167],[47,-0.1593],[48,-0.1526],[49,-0.1485],[50,-0.1447],[51,-0.1425],[52,-0.1418],[53,-0.1509],[54,-0.1574],[55,-0.1546],[56,-0.1574],[57,-0.1621],[58,-0.1659],[59,-0.1678],[60,-0.169],[61,-0.1666],[62,-0.1657],[63,-0.1638],[64,-0.1629],[65,-0.1636],[66,-0.1631],[67,-0.1618],[68,-0.1612],[69,-0.1601],[70,-0.1618],[71,-0.1689],[72,-0.1831],[73,-0.1902],[74,-0.1924],[75,-0.192],[76,-0.1922],[77,-0.1936],[78,-0.1941],[79,-0.1941],[80,-0.1954],[81,-0.1961],[82,-0.1959],[84,-0.1944],[85,-0.1913],[86,-0.1895],[87,-0.1866],[89,-0.1824],[91,-0.177],[94,-0.1754],[95,-0.1759],[96,-0.1798],[98,-0.184],[99,-0.1852],[101,-0.1843],[102,-0.1828],[103,-0.1827],[104,-0.1819],[105,-0.1816],[106,-0.1796],[107,-0.179],[109,-0.1793],[110,-0.1804],[112,-0.1849],[113,-0.1884],[114,-0.1894],[115,-0.1909],[116,-0.1895],[118,-0.1886],[119,-0.1877],[120,-0.1881],[121,-0.1875],[125,-0.1884],[126,-0.1881],[127,-0.1893],[128,-0.1912],[129,-0.1904],[130,-0.1918],[132,-0.1915],[133,-0.191],[134,-0.1876],[135,-0.1852],[136,-0.187],[137,-0.1882],[139,-0.1878],[140,-0.1881],[142,-0.1873],[143,-0.1851],[144,-0.1821],[145,-0.18],[146,-0.1802],[147,-0.1794],[148,-0.1796],[150,-0.1834],[151,-0.1875],[152,-0.1898],[153,-0.1948],[154,-0.1982],[156,-0.1972],[157,-0.196],[158,-0.1937],[159,-0.1904],[161,-0.1884],[162,-0.1864],[164,-0.1833],[166,-0.1864],[167,-0.1906],[168,-0.1704],[169,-0.1617],[170,-0.1571],[171,-0.151],[172,-0.1463],[173,-0.1403],[174,-0.1295],[175,-0.1086],[176,-0.0806],[177,-0.0448],[178,-0.0168],[179,0.0008],[180,0.0106],[181,0.0236],[182,-0.0017],[183,-0.1222],[184,-0.0482]]
       }
       ball_l_fk_ctrl.Rotation.Y {
        type=MovieSceneScriptingFloatChannel
        default=0.0902
        thinned_from=185
        keys [[0,0.3618],[2,0.3593],[3,0.3593],[4,0.3565],[5,0.3528],[6,0.3482],[7,0.3464],[9,0.3445],[11,0.3401],[12,0.3403],[13,0.3369],[14,0.3372],[15,0.3389],[16,0.3417],[17,0.3461],[18,0.3467],[19,0.35],[20,0.3514],[21,0.3565],[22,0.3626],[23,0.3593],[25,0.3594],[26,0.3617],[27,0.3539],[28,0.3538],[29,0.357],[31,0.361],[33,0.358],[34,0.3583],[35,0.3575],[36,0.3531],[37,0.3541],[41,0.3539],[42,0.3495],[43,0.3426],[44,0.3385],[45,0.3356],[47,0.3188],[48,0.3043],[49,0.2954],[50,0.2874],[51,0.2825],[52,0.2812],[53,0.3007],[54,0.3148],[55,0.3086],[56,0.3148],[57,0.3248],[58,0.3332],[59,0.3374],[60,0.3399],[61,0.3348],[62,0.3327],[63,0.3285],[64,0.3267],[65,0.3281],[66,0.327],[67,0.3243],[68,0.323],[69,0.3206],[70,0.3243],[71,0.3396],[72,0.3711],[73,0.3869],[74,0.3921],[75,0.3911],[76,0.3915],[77,0.3946],[78,0.3957],[79,0.3958],[80,0.3988],[81,0.4003],[82,0.3999],[83,0.3977],[84,0.3964],[85,0.3895],[86,0.3854],[87,0.3791],[89,0.3696],[91,0.3577],[94,0.3541],[95,0.3552],[96,0.3639],[98,0.3733],[99,0.3759],[101,0.3739],[102,0.3706],[103,0.3703],[104,0.3685],[105,0.3677],[106,0.3634],[107,0.3621],[109,0.3628],[110,0.3651],[112,0.3753],[113,0.383],[114,0.3854],[115,0.3887],[116,0.3855],[118,0.3835],[119,0.3815],[120,0.3823],[121,0.381],[125,0.383],[126,0.3824],[127,0.385],[128,0.3893],[129,0.3876],[130,0.3906],[132,0.39],[133,0.3889],[134,0.3812],[135,0.376],[136,0.3798],[137,0.3825],[139,0.3817],[140,0.3823],[142,0.3807],[143,0.3756],[144,0.369],[145,0.3644],[146,0.3646],[147,0.3629],[148,0.3634],[150,0.3719],[151,0.3809],[152,0.3861],[153,0.3974],[154,0.4051],[156,0.4028],[157,0.4],[158,0.3949],[159,0.3876],[161,0.3831],[162,0.3785],[164,0.3717],[166,0.3786],[167,0.388],[168,0.3431],[169,0.324],[170,0.314],[171,0.3009],[172,0.2907],[173,0.278],[174,0.255],[175,0.2115],[176,0.1543],[177,0.0836],[178,0.0296],[179,-0.0035],[180,-0.0218],[181,-0.0459],[182,0.0014],[183,0.2398],[184,0.0902]]
       }
       ball_l_fk_ctrl.Rotation.Z {
        type=MovieSceneScriptingFloatChannel
        default=1.1101
        thinned_from=185
        keys [[0,3.9423],[2,3.9168],[3,3.9171],[4,3.8881],[5,3.8504],[6,3.8027],[7,3.7837],[9,3.7643],[11,3.7185],[12,3.7212],[13,3.6857],[14,3.6887],[15,3.707],[16,3.7357],[17,3.7812],[18,3.787],[19,3.8211],[20,3.8359],[21,3.8884],[22,3.9514],[23,3.9173],[25,3.9178],[26,3.9422],[27,3.861],[28,3.8607],[29,3.8935],[31,3.9345],[33,3.9041],[34,3.9072],[35,3.8981],[36,3.8527],[37,3.8632],[41,3.8613],[42,3.8159],[43,3.7449],[44,3.7026],[45,3.6725],[47,3.4993],[48,3.3489],[49,3.257],[50,3.1734],[51,3.1233],[52,3.1093],[53,3.3114],[54,3.4574],[55,3.3935],[56,3.4571],[57,3.5613],[58,3.6479],[59,3.6914],[60,3.7171],[61,3.6641],[62,3.6431],[63,3.5995],[64,3.5802],[65,3.5949],[66,3.584],[67,3.5558],[68,3.5424],[69,3.5175],[70,3.5558],[71,3.7142],[72,4.0388],[73,4.2011],[74,4.2543],[75,4.244],[76,4.2487],[77,4.2808],[78,4.2918],[79,4.2924],[80,4.3233],[81,4.3391],[82,4.3343],[84,4.2991],[85,4.2282],[86,4.1859],[87,4.1204],[89,4.0235],[91,3.9003],[94,3.8639],[95,3.8743],[96,3.9642],[98,4.0608],[99,4.0879],[101,4.067],[102,4.0334],[103,4.0308],[104,4.012],[105,4.0038],[106,3.9591],[107,3.9457],[109,3.9527],[110,3.9772],[112,4.0814],[113,4.1612],[114,4.1853],[115,4.2196],[116,4.1871],[118,4.166],[119,4.1454],[120,4.1534],[121,4.1402],[125,4.1612],[126,4.1547],[127,4.1818],[128,4.2262],[129,4.2085],[130,4.2397],[132,4.2332],[133,4.2213],[134,4.1428],[135,4.0887],[136,4.1279],[137,4.1558],[139,4.1479],[140,4.1535],[142,4.1369],[143,4.0853],[144,4.0171],[145,3.9691],[146,3.972],[147,3.9543],[148,3.9597],[150,4.0469],[151,4.1392],[152,4.1935],[153,4.3091],[154,4.3879],[156,4.3643],[157,4.3362],[158,4.2836],[159,4.2085],[161,4.1617],[162,4.1142],[164,4.0448],[166,4.1159],[167,4.2128],[168,3.7499],[169,3.5526],[170,3.4494],[171,3.3136],[172,3.2083],[173,3.0766],[174,2.8377],[175,2.3842],[176,1.7857],[177,1.0412],[178,0.4701],[179,0.1182],[180,-0.0769],[181,-0.3345],[182,0.1685],[183,2.6788],[184,1.1101]]
       }
       ball_l_fk_ctrl.Scale.X {type=MovieSceneScriptingFloatChannel, default=1, const [1,185]}
       ball_l_fk_ctrl.Scale.Y {type=MovieSceneScriptingFloatChannel, default=1, const [1,185]}
       ball_l_fk_ctrl.Scale.Z {type=MovieSceneScriptingFloatChannel, default=1, const [1,185]}
       foot_l_ik_ctrl.Location.X {
        type=MovieSceneScriptingFloatChannel
        default=5.3331
        thinned_from=185
        keys [[0,37.9713],[2,38.305],[3,38.3442],[4,38.5309],[5,38.8849],[6,39.4687],[8,40.861],[9,41.2495],[10,41.9048],[12,42.9715],[13,43.1582],[14,43.1404],[17,42.103],[18,41.6035],[19,41.3519],[20,41.2133],[21,40.7232],[22,40.5793],[23,40.5596],[26,39.9934],[28,40.522],[30,40.2793],[31,40.428],[33,41.2325],[34,41.2926],[35,41.1736],[37,41.3823],[39,40.9272],[40,40.9248],[41,40.7165],[42,40.3427],[43,40.1961],[45,40.3684],[46,40.9618],[47,41.451],[48,41.8128],[49,41.9435],[52,41.5968],[53,40.2599],[54,39.2062],[55,37.3188],[56,36.4496],[57,35.2661],[58,34.4866],[59,33.4397],[60,32.1981],[61,30.6107],[62,29.6364],[63,28.4428],[66,25.7554],[67,24.5951],[68,22.9715],[69,21.624],[71,19.3587],[72,18.6519],[73,18.0389],[75,16.521],[76,15.7316],[77,14.7784],[78,13.7616],[79,12.8262],[80,12.0583],[81,11.3978],[82,10.9335],[83,10.5567],[84,10.1257],[86,9.2152],[88,8.1227],[89,7.4692],[90,6.8454],[91,6.3297],[92,5.9228],[93,5.5895],[94,5.2911],[95,5.0448],[96,4.9144],[98,4.927],[99,4.8833],[101,4.9457],[102,4.9551],[103,4.9521],[105,4.9792],[106,4.948],[108,4.8667],[109,4.9127],[111,5.0261],[112,5.0599],[113,5.1586],[115,5.3876],[117,5.4721],[119,5.5962],[121,5.6439],[122,5.6788],[123,5.7618],[126,5.8645],[127,5.9156],[128,5.9865],[129,6.0001],[132,5.9369],[133,5.9399],[134,5.929],[135,5.883],[136,5.8557],[137,5.8707],[140,5.9945],[141,5.9969],[142,5.9681],[143,5.9052],[146,5.7937],[147,5.7776],[149,5.7848],[150,5.8245],[152,5.9701],[153,6.1207],[154,6.328],[155,6.4566],[156,6.569],[157,6.6348],[159,6.8602],[160,7.0278],[161,7.0778],[162,7.2128],[163,7.2865],[164,7.2123],[165,7.0582],[166,6.9876],[167,6.9553],[169,6.993],[170,6.8192],[171,7.5099],[172,8.0954],[173,7.5657],[174,6.7596],[175,6.1259],[176,6.1592],[177,6.912],[178,8.7771],[179,10.4893],[180,13.8671],[181,18.3547],[182,21.9363],[183,15.2066],[184,5.3331]]
       }
       foot_l_ik_ctrl.Location.Y {
        type=MovieSceneScriptingFloatChannel
        default=24.6442
        thinned_from=185
        keys [[0,-95.1335],[5,-94.6038],[12,-92.7878],[14,-92.6893],[22,-93.9354],[26,-94.2406],[28,-94.1397],[31,-94.2719],[34,-93.944],[41,-93.9583],[45,-93.9564],[51,-91.778],[54,-89.6487],[55,-88.3013],[56,-86.5903],[57,-84.449],[58,-81.8222],[61,-73.0537],[62,-70.2445],[63,-67.2909],[65,-60.693],[67,-53.7179],[68,-50.4146],[69,-46.9647],[71,-39.8262],[72,-36.1586],[73,-32.4181],[74,-28.7851],[75,-25.2662],[76,-21.9571],[77,-18.9788],[78,-16.3375],[79,-13.9869],[80,-11.9369],[81,-10.2539],[82,-8.9292],[84,-6.6184],[85,-5.4739],[86,-4.3687],[87,-3.351],[88,-2.3737],[89,-1.4716],[90,-0.6213],[91,0.1988],[92,0.9856],[93,1.7157],[94,2.375],[95,2.951],[96,3.4492],[97,3.8668],[98,4.1577],[99,4.3044],[100,4.3287],[104,4.2606],[105,4.2523],[106,4.2253],[108,4.1901],[110,4.2029],[112,4.1852],[113,4.1954],[114,4.2242],[115,4.2711],[118,4.363],[119,4.3848],[122,4.3879],[124,4.4314],[125,4.4454],[129,4.4621],[131,4.4477],[132,4.4277],[134,4.4109],[135,4.3946],[137,4.3513],[139,4.3416],[140,4.3665],[141,4.3677],[143,4.3493],[145,4.3415],[147,4.2939],[148,4.2823],[149,4.2419],[151,4.2121],[152,4.2148],[153,4.2455],[154,4.3054],[157,4.4137],[160,4.5643],[161,4.5775],[162,4.6063],[163,4.6173],[164,4.5963],[165,4.5437],[167,4.4106],[168,4.4058],[169,4.2379],[170,2.9752],[171,1.1242],[172,-0.6211],[173,-2.0375],[174,-2.7134],[175,-2.542],[176,-2.1183],[177,-2.6697],[178,-3.6739],[179,-5.5244],[180,-7.6472],[181,-10.9076],[182,-13.1973],[183,15.7088],[184,24.6442]]
       }
       foot_l_ik_ctrl.Location.Z {
        type=MovieSceneScriptingFloatChannel
        default=89.8324
        thinned_from=185
        keys [[0,89.0211],[10,88.8176],[14,88.4188],[25,88.809],[47,87.506],[49,87.8286],[50,88.2052],[51,87.4971],[53,84.4957],[54,81.1535],[55,78.2195],[57,68.7707],[58,63.6564],[60,53.752],[62,44.3481],[63,40.3662],[64,36.2872],[65,32.135],[66,28.2904],[67,24.5875],[68,21.1708],[69,18.0371],[70,15.1769],[71,12.5806],[72,10.2983],[73,8.3095],[74,6.5376],[75,5.0083],[76,3.7288],[77,2.6604],[78,1.7843],[79,1.0819],[80,0.5461],[81,0.1539],[82,-0.1302],[83,-0.3627],[84,-0.5913],[85,-0.7962],[86,-0.9697],[87,-1.1286],[88,-1.2705],[89,-1.3992],[90,-1.5152],[91,-1.6186],[92,-1.7148],[93,-1.7787],[94,-1.8374],[95,-1.8751],[96,-1.9026],[97,-1.9187],[99,-1.9098],[100,-1.9094],[103,-1.9212],[104,-1.9521],[105,-1.954],[106,-1.9141],[107,-1.9203],[108,-1.9393],[110,-1.9549],[114,-1.9995],[116,-2.0101],[117,-2.0019],[118,-1.989],[119,-1.9802],[121,-1.9808],[124,-1.9907],[125,-1.9968],[128,-2.042],[130,-2.0398],[133,-2.0259],[134,-2.0084],[136,-1.9538],[137,-1.9414],[138,-1.9404],[141,-1.9634],[142,-1.9639],[143,-1.9554],[144,-1.9398],[145,-1.9124],[146,-1.8898],[147,-1.8784],[149,-1.8762],[150,-1.8893],[151,-1.9202],[152,-1.9561],[154,-2.0345],[156,-2.064],[157,-2.0609],[158,-2.0456],[160,-1.997],[161,-1.9859],[163,-1.9529],[164,-1.9397],[165,-1.9431],[166,-1.989],[167,-2.1436],[168,-2.8394],[169,-3.893],[170,-5.2267],[171,-6.4593],[172,-7.3861],[173,-8.4583],[174,-9.3298],[175,-9.9792],[176,-10.2259],[177,-10.2055],[178,-10.0262],[179,-9.7269],[180,-9.224],[181,-8.2833],[182,1.177],[183,51.3272],[184,89.8324]]
       }
       foot_l_ik_ctrl.Rotation.X {
        type=MovieSceneScriptingFloatChannel
        default=97.613
        thinned_from=185
        keys [[0,124.6306],[25,125.2643],[33,123.6455],[46,124.8045],[50,125.7795],[51,125.4421],[52,124.5745],[53,123.0099],[55,118.5541],[56,115.7383],[59,105.2997],[66,78.0187],[69,67.3608],[70,63.9564],[73,54.3391],[74,51.3315],[75,48.6177],[77,43.639],[78,41.4351],[79,39.4867],[80,37.7077],[81,36.1978],[82,35.027],[85,32.2024],[87,30.2638],[89,28.4842],[91,26.8691],[92,26.148],[93,25.5442],[94,25.0318],[95,24.5764],[96,24.1807],[97,23.8843],[98,23.6944],[99,23.5673],[100,23.5623],[102,23.635],[105,23.6215],[106,23.75],[110,23.797],[114,23.6829],[115,23.6976],[117,23.8341],[123,23.9128],[126,24.0482],[131,24.078],[133,24.0503],[134,24.0789],[135,24.0353],[137,23.8872],[145,23.9784],[151,23.7924],[153,23.8902],[154,24.0549],[160,24.7005],[162,24.8675],[163,24.8905],[164,24.8486],[166,24.5484],[167,24.4402],[168,25.3963],[169,25.2418],[170,26.203],[171,28.4371],[172,30.6093],[173,33.1313],[174,35.0048],[175,35.816],[176,36.1616],[177,37.3338],[178,38.3909],[179,39.9075],[180,41.174],[181,42.713],[182,49.017],[183,72.1457],[184,97.613]]
       }
       foot_l_ik_ctrl.Rotation.Y {
        type=MovieSceneScriptingFloatChannel
        default=4.3602
        thinned_from=185
        keys [[0,-4.4887],[1,-4.2789],[2,-4.1183],[3,-4.0373],[4,-4.1392],[5,-4.3447],[6,-4.4811],[7,-4.4568],[8,-4.3723],[9,-4.314],[10,-4.3732],[11,-4.5111],[12,-4.6798],[13,-4.7942],[14,-4.7402],[15,-4.6694],[16,-4.51],[17,-4.3323],[18,-4.1643],[19,-3.9558],[20,-3.8335],[21,-3.4976],[22,-3.1317],[23,-2.7723],[24,-2.5175],[25,-2.3824],[26,-2.4234],[27,-2.6992],[28,-2.8923],[29,-2.9959],[30,-3.2063],[32,-3.8821],[33,-4.1478],[35,-4.2905],[36,-4.3952],[37,-4.4727],[38,-4.4531],[39,-4.4882],[40,-4.5468],[41,-4.4943],[42,-4.3766],[43,-4.2967],[45,-4.3457],[46,-4.3388],[47,-4.1575],[48,-3.6831],[49,-3.0088],[50,-2.2303],[51,-1.6179],[52,-0.6157],[53,0.8458],[54,1.8792],[55,2.745],[57,4.1183],[58,4.9219],[59,5.626],[60,5.9617],[61,5.9638],[62,6.2525],[63,6.7799],[64,7.572],[65,8.0749],[66,8.5544],[67,8.606],[68,8.378],[69,8.1716],[70,7.7362],[71,7.2578],[72,6.3925],[73,5.926],[74,5.7158],[75,5.5607],[76,5.2732],[78,4.7193],[79,4.4769],[81,3.9194],[82,3.585],[83,3.4534],[84,3.5353],[85,3.6639],[86,3.6649],[87,3.7544],[88,3.9162],[89,4.3355],[90,4.6866],[91,4.8597],[92,4.8755],[93,4.7982],[94,4.7326],[95,4.6024],[96,4.2755],[97,3.7769],[98,3.3489],[99,3.114],[100,2.8456],[101,2.7104],[103,2.5738],[104,2.4744],[105,2.4129],[106,2.4636],[107,2.5463],[108,2.6093],[109,2.5409],[110,2.413],[111,2.2299],[112,2.0584],[113,1.7272],[114,1.4625],[115,1.1284],[116,1.0406],[117,0.9704],[118,0.8751],[119,0.7539],[120,0.7012],[121,0.6559],[122,0.5593],[123,0.3824],[124,0.3108],[125,0.2086],[126,0.1335],[127,-0.0481],[128,-0.2743],[129,-0.341],[130,-0.28],[131,-0.268],[132,-0.2667],[133,-0.2828],[134,-0.1901],[135,-0.1156],[136,-0.1448],[137,-0.2516],[138,-0.3788],[139,-0.5122],[140,-0.5625],[141,-0.5687],[142,-0.5201],[143,-0.3713],[144,-0.215],[145,-0.0944],[146,-0.0446],[147,-0.0591],[148,-0.0934],[149,-0.2488],[150,-0.4245],[151,-0.7285],[152,-0.9876],[153,-1.4954],[154,-2.102],[155,-2.4239],[156,-2.687],[157,-2.8002],[158,-3.0453],[159,-3.2768],[160,-3.6663],[161,-3.8159],[162,-4.1665],[163,-4.3436],[164,-4.1025],[165,-3.7664],[166,-3.5956],[167,-3.3869],[168,-3.4507],[169,-2.2847],[170,-0.086],[171,0.4158],[172,0.0818],[173,0.2966],[174,0.3344],[175,0.1735],[176,0.1222],[177,0.3302],[178,0.3442],[179,-0.0521],[180,-0.3493],[181,-0.1973],[182,1.3319],[183,6.2037],[184,4.3602]]
       }
       foot_l_ik_ctrl.Rotation.Z {
        type=MovieSceneScriptingFloatChannel
        default=-3.3106
        thinned_from=185
        keys [[0,23.9915],[2,23.9941],[3,24.137],[4,24.6484],[6,25.8547],[7,26.1909],[8,26.3996],[9,26.6614],[11,27.9469],[12,28.5347],[13,28.7902],[14,28.6077],[15,28.1618],[16,27.6026],[17,26.9823],[18,26.2787],[19,25.7332],[20,25.4625],[21,24.7748],[22,24.262],[23,23.9155],[24,23.6812],[25,23.6927],[26,23.9463],[28,25.2031],[29,25.5022],[30,25.9831],[31,26.8037],[32,27.5247],[33,28.0096],[34,28.1306],[36,28.2283],[37,28.1614],[40,27.3273],[41,26.9188],[42,26.4479],[43,26.111],[44,26.062],[45,26.1155],[46,26.2791],[47,26.2809],[48,26.0439],[49,25.7322],[50,25.1564],[52,24.6476],[53,23.8073],[55,23.5052],[56,23.9473],[57,24.1485],[58,24.0247],[59,23.1046],[60,22.2346],[61,21.9094],[62,21.7571],[63,21.5058],[64,21.3707],[65,21.0592],[66,20.3122],[67,19.3225],[68,18.2228],[69,17.0593],[70,15.7292],[71,14.8558],[72,13.4954],[73,12.6315],[74,12.2397],[75,12.0018],[76,11.6989],[77,11.467],[79,11.1147],[81,10.6827],[82,10.416],[83,10.2886],[85,10.2617],[86,10.173],[87,10.1623],[88,10.2159],[89,10.4],[90,10.5487],[91,10.5855],[92,10.5287],[95,10.2693],[96,10.1251],[98,9.7124],[100,9.5198],[101,9.4564],[105,9.3574],[106,9.4408],[108,9.5267],[109,9.5263],[110,9.4898],[112,9.3548],[113,9.1863],[114,9.057],[115,8.8645],[116,8.8165],[118,8.7533],[119,8.7086],[121,8.6574],[122,8.5991],[123,8.4964],[126,8.3355],[127,8.2293],[128,8.0999],[129,8.0704],[131,8.1249],[133,8.1261],[134,8.1833],[135,8.2202],[136,8.2258],[137,8.1842],[139,8.0456],[140,8.0275],[142,8.0622],[144,8.2431],[145,8.3187],[146,8.3611],[148,8.3608],[149,8.314],[150,8.2182],[152,7.86],[153,7.5396],[154,7.1496],[155,6.9527],[156,6.7839],[157,6.7107],[159,6.4465],[160,6.2225],[161,6.1295],[162,5.9336],[163,5.8524],[164,6.0055],[165,6.1841],[166,6.28],[167,6.3029],[168,5.9818],[169,6.6221],[170,8.402],[171,9.5901],[172,10.9984],[173,12.6729],[174,14.0469],[175,14.0929],[176,13.7304],[177,14.562],[178,15.8795],[179,17.7828],[180,18.4998],[181,16.8642],[182,8.872],[183,-7.298],[184,-3.3106]]
       }
       foot_l_ik_ctrl.Scale.X {type=MovieSceneScriptingFloatChannel, default=1, const [1,185]}
       foot_l_ik_ctrl.Scale.Y {type=MovieSceneScriptingFloatChannel, default=1, const [1,185]}
       foot_l_ik_ctrl.Scale.Z {type=MovieSceneScriptingFloatChannel, default=1, const [1,185]}
       ball_l_ik_ctrl.Location.X {type=MovieSceneScriptingFloatChannel, default=0.0005, thinned_from=185, keys [[0,0],[183,0],[184,0.0005]]}
       ball_l_ik_ctrl.Location.Y {type=MovieSceneScriptingFloatChannel, default=-0.0013, thinned_from=185, keys [[0,0],[182,0],[183,-0.0002],[184,-0.0013]]}
       ball_l_ik_ctrl.Location.Z {type=MovieSceneScriptingFloatChannel, default=0.0008, thinned_from=185, keys [[0,0],[182,0],[184,0.0008]]}
       ball_l_ik_ctrl.Rotation.X {
        type=MovieSceneScriptingFloatChannel
        default=1.1137
        thinned_from=185
        keys [[0,3.9587],[2,3.9331],[3,3.9334],[4,3.9042],[5,3.8663],[6,3.8185],[7,3.7994],[9,3.7798],[11,3.7339],[12,3.7365],[13,3.7009],[14,3.7039],[15,3.7223],[16,3.7512],[17,3.7969],[18,3.8027],[19,3.8369],[20,3.8518],[21,3.9045],[22,3.9678],[23,3.9336],[25,3.9341],[26,3.9586],[27,3.877],[28,3.8767],[29,3.9096],[31,3.9509],[33,3.9203],[34,3.9234],[35,3.9143],[36,3.8686],[37,3.8792],[41,3.8773],[42,3.8317],[43,3.7603],[44,3.7179],[45,3.6877],[47,3.5136],[48,3.3625],[49,3.2703],[50,3.1863],[51,3.1359],[52,3.1218],[53,3.3249],[54,3.4716],[55,3.4073],[56,3.4712],[57,3.576],[58,3.6629],[59,3.7066],[60,3.7325],[61,3.6792],[62,3.6581],[63,3.6143],[64,3.5949],[65,3.6097],[66,3.5988],[67,3.5704],[68,3.557],[69,3.532],[70,3.5704],[71,3.7295],[72,4.0556],[73,4.2187],[74,4.2722],[75,4.2618],[76,4.2665],[77,4.2987],[78,4.3098],[79,4.3104],[80,4.3415],[81,4.3573],[82,4.3525],[84,4.3171],[85,4.2459],[86,4.2034],[87,4.1377],[89,4.0403],[91,3.9164],[94,3.8799],[95,3.8903],[96,3.9806],[98,4.0777],[99,4.105],[101,4.084],[102,4.0502],[103,4.0476],[104,4.0287],[105,4.0205],[106,3.9755],[107,3.9621],[109,3.9691],[110,3.9937],[112,4.0985],[113,4.1786],[114,4.2028],[115,4.2373],[116,4.2046],[118,4.1834],[119,4.1627],[120,4.1707],[121,4.1575],[125,4.1786],[126,4.1721],[127,4.1993],[128,4.2439],[129,4.2261],[130,4.2575],[132,4.2509],[133,4.239],[134,4.1601],[135,4.1057],[136,4.1451],[137,4.1732],[139,4.1652],[140,4.1709],[142,4.1542],[143,4.1023],[144,4.0339],[145,3.9856],[146,3.9885],[147,3.9707],[148,3.9761],[150,4.0638],[151,4.1565],[152,4.211],[153,4.3272],[154,4.4064],[156,4.3826],[157,4.3544],[158,4.3015],[159,4.2261],[161,4.1791],[162,4.1314],[164,4.0616],[166,4.1331],[167,4.2304],[168,3.7654],[169,3.5671],[170,3.4636],[171,3.3271],[172,3.2213],[173,3.089],[174,2.8491],[175,2.3935],[176,1.7923],[177,1.0445],[178,0.4709],[179,0.1175],[180,-0.0785],[181,-0.3371],[182,0.168],[183,2.6894],[184,1.1137]]
       }
       ball_l_ik_ctrl.Rotation.Y {
        type=MovieSceneScriptingFloatChannel
        default=-0.007
        thinned_from=185
        keys [[0,0.007],[14,0.006],[21,0.0068],[22,0.007],[25,0.0069],[26,0.007],[27,0.0067],[35,0.0068],[36,0.0067],[41,0.0067],[45,0.006],[47,0.0053],[48,0.0047],[50,0.0039],[52,0.0036],[53,0.0045],[54,0.0051],[55,0.0049],[58,0.0059],[60,0.0062],[64,0.0056],[69,0.0054],[70,0.0055],[71,0.0061],[72,0.0074],[73,0.0079],[74,0.0081],[82,0.0084],[84,0.0083],[91,0.0068],[95,0.0068],[99,0.0075],[109,0.007],[115,0.008],[126,0.0078],[128,0.008],[133,0.008],[135,0.0075],[142,0.0077],[145,0.0071],[148,0.0071],[150,0.0074],[154,0.0086],[158,0.0082],[164,0.0074],[166,0.0076],[167,0.008],[168,0.0063],[169,0.0055],[173,0.0035],[174,0.0024],[175,0.0002],[176,-0.003],[177,-0.0074],[178,-0.0111],[179,-0.0136],[180,-0.015],[181,-0.0169],[182,-0.0131],[183,0.0017],[184,-0.007]]
       }
       ball_l_ik_ctrl.Rotation.Z {
        type=MovieSceneScriptingFloatChannel
        default=0.0498
        thinned_from=185
        keys [[0,0.1946],[2,0.1933],[3,0.1934],[4,0.1919],[5,0.1899],[6,0.1875],[7,0.1865],[9,0.1855],[11,0.1832],[12,0.1833],[13,0.1815],[14,0.1816],[15,0.1826],[16,0.184],[17,0.1864],[18,0.1867],[19,0.1884],[20,0.1892],[21,0.1919],[22,0.1951],[23,0.1934],[25,0.1934],[26,0.1946],[27,0.1905],[28,0.1904],[29,0.1921],[31,0.1942],[33,0.1927],[34,0.1929],[35,0.1924],[36,0.19],[37,0.1906],[41,0.1905],[42,0.1881],[43,0.1845],[44,0.1824],[45,0.1808],[47,0.1719],[48,0.1642],[49,0.1595],[50,0.1552],[51,0.1527],[52,0.1519],[53,0.1623],[54,0.1698],[55,0.1665],[56,0.1698],[57,0.1751],[58,0.1796],[59,0.1818],[60,0.1831],[61,0.1804],[62,0.1793],[63,0.1771],[64,0.1761],[65,0.1768],[66,0.1763],[67,0.1748],[68,0.1741],[69,0.1729],[70,0.1748],[71,0.183],[72,0.1996],[73,0.2079],[74,0.2106],[75,0.2101],[76,0.2104],[77,0.212],[78,0.2126],[79,0.2126],[80,0.2142],[81,0.215],[82,0.2148],[83,0.2136],[84,0.2129],[85,0.2093],[86,0.2071],[87,0.2038],[89,0.1988],[91,0.1925],[94,0.1906],[95,0.1912],[96,0.1958],[98,0.2007],[99,0.2021],[101,0.201],[102,0.1993],[103,0.1992],[104,0.1982],[105,0.1978],[106,0.1955],[107,0.1948],[109,0.1952],[110,0.1964],[112,0.2018],[113,0.2059],[114,0.2071],[115,0.2089],[116,0.2072],[118,0.2061],[119,0.2051],[120,0.2055],[121,0.2048],[125,0.2059],[126,0.2055],[127,0.2069],[128,0.2092],[129,0.2083],[130,0.2099],[132,0.2096],[133,0.209],[134,0.2049],[135,0.2021],[136,0.2042],[137,0.2056],[139,0.2052],[140,0.2055],[142,0.2046],[143,0.202],[144,0.1985],[145,0.196],[146,0.1962],[147,0.1953],[148,0.1955],[150,0.2],[151,0.2047],[152,0.2075],[153,0.2135],[154,0.2175],[156,0.2163],[157,0.2148],[158,0.2121],[159,0.2083],[161,0.2059],[162,0.2035],[164,0.1999],[166,0.2035],[167,0.2085],[168,0.1847],[169,0.1747],[170,0.1694],[171,0.1624],[172,0.157],[173,0.1503],[174,0.1381],[175,0.1148],[176,0.0843],[177,0.0463],[178,0.0171],[179,-0.0008],[180,-0.0107],[181,-0.0239],[182,0.0017],[183,0.1299],[184,0.0498]]
       }
       ball_l_ik_ctrl.Scale.X {type=MovieSceneScriptingFloatChannel, default=1, const [1,185]}
       ball_l_ik_ctrl.Scale.Y {type=MovieSceneScriptingFloatChannel, default=1, const [1,185]}
       ball_l_ik_ctrl.Scale.Z {type=MovieSceneScriptingFloatChannel, default=1, const [1,185]}
       foot_bk1_l_ctrl.Location.X {type=MovieSceneScriptingFloatChannel, default=0, const [0,185]}
       foot_bk1_l_ctrl.Location.Y {type=MovieSceneScriptingFloatChannel, default=0, const [0,185]}
       foot_bk1_l_ctrl.Location.Z {type=MovieSceneScriptingFloatChannel, default=0, const [0,185]}
       foot_bk1_l_ctrl.Rotation.X {type=MovieSceneScriptingFloatChannel, default=0, const [0,185]}
       foot_bk1_l_ctrl.Rotation.Y {type=MovieSceneScriptingFloatChannel, default=0, const [0,185]}
       foot_bk1_l_ctrl.Rotation.Z {type=MovieSceneScriptingFloatChannel, default=0, const [0,185]}
       foot_bk1_l_ctrl.Scale.X {type=MovieSceneScriptingFloatChannel, default=1, const [1,185]}
       foot_bk1_l_ctrl.Scale.Y {type=MovieSceneScriptingFloatChannel, default=1, const [1,185]}
       foot_bk1_l_ctrl.Scale.Z {type=MovieSceneScriptingFloatChannel, default=1, const [1,185]}
       foot_roll_l_ctrl.Location.X {type=MovieSceneScriptingFloatChannel, default=0, const [0,185]}
       foot_roll_l_ctrl.Location.Y {type=MovieSceneScriptingFloatChannel, default=0, const [0,185]}
       foot_roll_l_ctrl.Location.Z {type=MovieSceneScriptingFloatChannel, default=0, const [0,185]}
       foot_roll_l_ctrl.Rotation.X {type=MovieSceneScriptingFloatChannel, default=0, const [0,185]}
       foot_roll_l_ctrl.Rotation.Y {type=MovieSceneScriptingFloatChannel, default=0, const [0,185]}
       foot_roll_l_ctrl.Rotation.Z {type=MovieSceneScriptingFloatChannel, default=0, const [0,185]}
       foot_roll_l_ctrl.Scale.X {type=MovieSceneScriptingFloatChannel, default=1, const [1,185]}
       foot_roll_l_ctrl.Scale.Y {type=MovieSceneScriptingFloatChannel, default=1, const [1,185]}
       foot_roll_l_ctrl.Scale.Z {type=MovieSceneScriptingFloatChannel, default=1, const [1,185]}
       roll_blend_l_ctrl {type=MovieSceneScriptingFloatChannel, default=1, const [1,185]}
       tip_l_ctrl.Location.X {type=MovieSceneScriptingFloatChannel, default=0, const [0,185]}
       tip_l_ctrl.Location.Y {type=MovieSceneScriptingFloatChannel, default=0, const [0,185]}
       tip_l_ctrl.Location.Z {type=MovieSceneScriptingFloatChannel, default=0, const [0,185]}
       tip_l_ctrl.Rotation.X {type=MovieSceneScriptingFloatChannel, default=0, const [0,185]}
       tip_l_ctrl.Rotation.Y {type=MovieSceneScriptingFloatChannel, default=0, const [0,185]}
       tip_l_ctrl.Rotation.Z {type=MovieSceneScriptingFloatChannel, default=0, const [0,185]}
       tip_l_ctrl.Scale.X {type=MovieSceneScriptingFloatChannel, default=1, const [1,185]}
       tip_l_ctrl.Scale.Y {type=MovieSceneScriptingFloatChannel, default=1, const [1,185]}
       tip_l_ctrl.Scale.Z {type=MovieSceneScriptingFloatChannel, default=1, const [1,185]}
       heel_l_ctrl.Location.X {type=MovieSceneScriptingFloatChannel, default=0, const [0,185]}
       heel_l_ctrl.Location.Y {type=MovieSceneScriptingFloatChannel, default=0, const [0,185]}
       heel_l_ctrl.Location.Z {type=MovieSceneScriptingFloatChannel, default=0, const [0,185]}
       heel_l_ctrl.Rotation.X {type=MovieSceneScriptingFloatChannel, default=0, const [0,185]}
       heel_l_ctrl.Rotation.Y {type=MovieSceneScriptingFloatChannel, default=0, const [0,185]}
       heel_l_ctrl.Rotation.Z {type=MovieSceneScriptingFloatChannel, default=0, const [0,185]}
       heel_l_ctrl.Scale.X {type=MovieSceneScriptingFloatChannel, default=1, const [1,185]}
       heel_l_ctrl.Scale.Y {type=MovieSceneScriptingFloatChannel, default=1, const [1,185]}
       heel_l_ctrl.Scale.Z {type=MovieSceneScriptingFloatChannel, default=1, const [1,185]}
       leg_l_softness {type=MovieSceneScriptingFloatChannel, default=0, const [0,185]}
       leg_l_pv_ik_ctrl.Location.X {
        type=MovieSceneScriptingFloatChannel
        default=3.9034
        thinned_from=185
        keys [[0,-2.6562],[1,-2.2592],[2,-1.9948],[3,-1.9497],[4,-1.9645],[5,-2.0007],[6,-1.7975],[7,-1.2705],[8,-0.7113],[9,-0.4883],[10,-0.3049],[11,-0.2561],[12,-0.2445],[13,-0.2163],[14,-0.029],[15,-0.0446],[16,0.0178],[17,0.0142],[18,-0.0673],[19,0.0595],[20,0.129],[21,0.1388],[22,0.3171],[23,0.5356],[24,0.5804],[25,0.5538],[26,0.3321],[27,0.1956],[28,0.0537],[29,-0.2862],[30,-0.7412],[31,-1.2734],[32,-1.551],[33,-1.6389],[34,-1.6703],[35,-1.7575],[36,-1.6132],[37,-1.5387],[38,-1.6327],[39,-1.7048],[40,-1.5395],[41,-1.4482],[42,-1.384],[43,-1.3131],[44,-1.2959],[45,-1.2373],[46,-0.6542],[47,0.072],[49,1.8978],[50,2.677],[51,3.0034],[52,3.429],[53,3.5263],[54,3.0013],[55,1.6019],[56,0.6057],[57,-0.6785],[58,-1.4775],[59,-1.9634],[60,-2.6416],[61,-4.0416],[62,-5.0245],[63,-6.0186],[64,-6.7698],[65,-7.4228],[66,-7.6331],[67,-7.8041],[68,-8.1492],[69,-8.1026],[70,-7.5887],[71,-7.6339],[72,-7.1427],[73,-7.0709],[74,-7.3789],[75,-7.6287],[76,-7.9678],[78,-8.9062],[79,-9.3436],[80,-9.7096],[81,-10.0335],[82,-10.1025],[83,-10.101],[85,-10.6509],[86,-10.796],[87,-11.1569],[88,-11.6345],[89,-12.4856],[90,-13.2053],[91,-13.6612],[92,-13.8766],[93,-13.9702],[94,-14.0287],[95,-14.023],[96,-13.7679],[97,-13.2474],[98,-12.8082],[99,-12.6419],[100,-12.3013],[101,-12.1361],[102,-12.0356],[103,-12.0144],[104,-11.8958],[105,-11.838],[106,-11.8734],[108,-12.0789],[110,-11.7352],[111,-11.5827],[112,-11.5151],[113,-11.3037],[114,-11.0393],[115,-10.9072],[116,-10.894],[117,-10.9198],[118,-10.8429],[119,-10.6959],[120,-10.6004],[121,-10.5448],[122,-10.4329],[123,-10.2209],[124,-10.14],[125,-10.1128],[126,-10.0217],[127,-9.9049],[128,-9.7052],[129,-9.6731],[131,-9.8041],[132,-9.8181],[133,-9.7657],[134,-9.6268],[136,-9.6876],[137,-9.6065],[140,-9.2692],[141,-9.2282],[142,-9.2739],[143,-9.3903],[144,-9.4319],[145,-9.4354],[146,-9.4741],[148,-9.3867],[149,-9.3266],[150,-9.2265],[152,-8.9264],[153,-8.7011],[154,-8.4274],[155,-8.2714],[156,-8.1493],[157,-8.1387],[158,-8.0266],[159,-7.86],[160,-7.5702],[161,-7.4924],[162,-7.2088],[163,-7.0345],[164,-7.1302],[165,-7.4648],[167,-7.7026],[168,-7.9701],[169,-7.5448],[170,-7.6578],[171,-6.3566],[172,-4.7515],[173,6.7474],[174,6.853],[175,6.7665],[176,9.9472],[177,9.9242],[178,7.3441],[179,9.9754],[180,8.1367],[181,13.1499],[182,21.0025],[183,14.634],[184,3.9034]]
       }
       leg_l_pv_ik_ctrl.Location.Y {
        type=MovieSceneScriptingFloatChannel
        default=20.7699
        thinned_from=185
        keys [[0,-99.3642],[14,-97.8665],[28,-98.4103],[32,-98.7547],[42,-97.9756],[45,-97.3905],[49,-95.5387],[50,-95.2642],[51,-94.2476],[52,-92.0822],[53,-89.6771],[54,-85.9293],[55,-82.3596],[57,-72.1616],[58,-66.7177],[59,-61.5801],[60,-56.6958],[62,-48.4453],[63,-45.3667],[64,-41.9373],[66,-34.3572],[67,-30.3352],[68,-26.6548],[69,-23.2955],[70,-20.1013],[71,-17.1184],[72,-14.4552],[73,-12.1086],[74,-9.8594],[75,-7.6805],[76,-5.7376],[77,-4.1087],[78,-2.7226],[79,-1.533],[80,-0.617],[81,0.0526],[82,0.5708],[83,1.0416],[84,1.5759],[85,2.0507],[86,2.4154],[87,2.7367],[88,3.0708],[89,3.3643],[91,3.9273],[92,4.2372],[93,4.5304],[94,4.8132],[95,5.0338],[96,5.223],[97,5.3806],[98,5.4661],[99,5.4813],[100,5.4692],[102,5.4768],[103,5.4722],[104,5.5134],[105,5.5151],[106,5.4765],[107,5.493],[108,5.5312],[113,5.6223],[114,5.6591],[115,5.7209],[116,5.8062],[117,5.8604],[118,5.8827],[121,5.8992],[122,5.9185],[125,6.047],[126,6.1052],[127,6.1379],[128,6.1914],[130,6.2048],[131,6.1796],[132,6.1286],[133,6.101],[134,6.0135],[136,5.7216],[137,5.6334],[138,5.5871],[139,5.5803],[140,5.6071],[141,5.6094],[142,5.5868],[144,5.4975],[146,5.3166],[147,5.2388],[148,5.2133],[149,5.1591],[150,5.1495],[151,5.216],[152,5.3238],[153,5.4912],[154,5.7321],[156,6.0209],[157,6.1197],[158,6.2],[160,6.2711],[162,6.2716],[163,6.2265],[165,6.0652],[166,6.0399],[167,6.2871],[168,7.6732],[169,8.728],[170,9.3993],[171,9.9763],[172,10.4237],[173,11.5488],[174,13.0256],[175,13.0535],[176,13.6418],[177,13.8766],[178,13.7883],[179,13.0426],[180,12.042],[181,10.4052],[182,10.8664],[183,19.9601],[184,20.7699]]
       }
       leg_l_pv_ik_ctrl.Location.Z {
        type=MovieSceneScriptingFloatChannel
        default=73.1607
        thinned_from=185
        keys [[0,-7.4963],[1,-7.5291],[5,-7.5984],[6,-7.5955],[7,-7.5475],[9,-7.416],[10,-7.4103],[11,-7.4642],[14,-7.7371],[15,-7.7554],[17,-7.7295],[22,-7.7916],[23,-7.8318],[24,-7.8505],[27,-7.8652],[28,-7.796],[29,-7.6929],[30,-7.6334],[32,-7.6383],[33,-7.6653],[35,-7.6746],[37,-7.7739],[38,-7.8369],[39,-7.9171],[40,-8.022],[42,-8.1655],[43,-8.31],[44,-8.493],[45,-8.6095],[46,-8.8914],[47,-9.1089],[48,-9.2683],[49,-9.235],[50,-8.9724],[51,-9.2397],[52,-9.7594],[53,-9.7149],[54,-10.5645],[55,-10.4692],[56,-11.2221],[57,-11.5744],[58,-11.7239],[59,-11.4876],[60,-11.1911],[61,-10.818],[62,-10.5157],[63,-10.0239],[64,-9.5002],[65,-8.8493],[66,-8.0662],[67,-7.0503],[68,-5.9169],[69,-4.6541],[70,-3.2417],[71,-1.6902],[72,-0.1517],[73,1.3608],[74,2.9189],[75,4.5643],[76,6.1601],[77,7.5964],[78,8.9115],[79,10.1064],[80,11.0683],[81,11.808],[82,12.3871],[83,12.9166],[84,13.5507],[85,14.1323],[87,15.0897],[90,16.6662],[92,17.6708],[94,18.556],[95,18.9019],[96,19.1758],[97,19.372],[98,19.5073],[99,19.566],[102,19.5508],[106,19.4526],[108,19.5422],[112,19.5282],[115,19.3628],[116,19.399],[120,19.2412],[122,19.2267],[128,19.4456],[130,19.4525],[134,19.3219],[136,19.0584],[137,18.9919],[138,18.975],[139,19.003],[141,19.1396],[142,19.1681],[144,19.1107],[147,18.8337],[149,18.7693],[150,18.7657],[151,18.8511],[153,19.139],[154,19.3681],[156,19.574],[158,19.6146],[161,19.5945],[163,19.6113],[165,19.5356],[166,19.692],[167,20.5869],[168,24.3792],[169,27.4776],[170,29.1399],[171,29.1745],[172,27.9253],[173,2.5768],[174,18.6469],[175,1.6562],[176,23.6704],[177,23.2207],[178,0.7292],[179,23.9445],[180,-2.361],[181,-4.6407],[182,-5.4296],[183,51.0836],[184,73.1607]]
       }
       leg_l_pv_ik_ctrl.Rotation.X {
        type=MovieSceneScriptingFloatChannel
        default=35.1085
        thinned_from=185
        keys [[0,72.3707],[11,72.1238],[15,71.844],[28,71.8782],[32,72.0271],[42,71.3884],[45,70.8844],[48,70.2306],[49,70.2297],[50,70.4196],[51,69.6825],[52,68.0556],[53,66.5729],[54,63.5326],[55,61.3004],[56,57.4799],[57,53.9332],[58,50.1582],[60,43.5411],[61,40.8846],[62,37.676],[63,35.7801],[64,33.4624],[65,30.3811],[66,27.3998],[67,23.8821],[68,20.5204],[69,17.4129],[71,11.3887],[72,8.7576],[74,4.2182],[75,1.9145],[76,-0.2061],[77,-2.0632],[78,-3.7385],[79,-5.2263],[80,-6.3635],[81,-7.2043],[82,-7.8619],[83,-8.4697],[84,-9.2227],[85,-9.8941],[86,-10.4239],[87,-10.9819],[88,-11.579],[90,-12.8696],[92,-14.1],[94,-15.1362],[95,-15.5334],[96,-15.8472],[97,-16.0654],[98,-16.2067],[99,-16.271],[101,-16.3098],[102,-16.2937],[103,-16.2367],[105,-16.2481],[106,-16.1795],[107,-16.232],[108,-16.3228],[112,-16.3609],[114,-16.2356],[115,-16.1294],[116,-16.1541],[117,-16.1002],[119,-15.9119],[120,-15.858],[122,-15.8408],[124,-15.8985],[126,-16.0505],[128,-16.1355],[130,-16.137],[132,-16.056],[133,-16.0411],[134,-15.9922],[136,-15.6655],[137,-15.5934],[138,-15.5891],[139,-15.644],[141,-15.8453],[142,-15.895],[144,-15.8161],[146,-15.5407],[147,-15.4569],[149,-15.4035],[150,-15.4233],[151,-15.5614],[153,-15.9581],[154,-16.2271],[155,-16.3475],[156,-16.4325],[157,-16.4304],[158,-16.3809],[159,-16.2861],[160,-16.232],[163,-16.2031],[164,-16.1388],[165,-16.1466],[166,-16.4421],[167,-17.815],[168,-23.6765],[169,-29.4879],[170,-34.3964],[171,-37.3742],[172,-37.9505],[173,-21.6945],[174,-24.749],[175,-24.6989],[176,-28.9391],[177,-28.9156],[178,-25.3729],[179,-27.9444],[180,-21.3927],[181,-17.7277],[182,-4.9675],[183,13.519],[184,35.1085]]
       }
       leg_l_pv_ik_ctrl.Rotation.Y {
        type=MovieSceneScriptingFloatChannel
        default=2.2006
        thinned_from=185
        keys [[0,-5.8438],[1,-5.4477],[2,-5.2316],[3,-5.2419],[4,-5.4052],[5,-5.6798],[6,-5.7562],[7,-5.4807],[8,-5.1394],[9,-5.1009],[10,-5.2482],[12,-5.7394],[13,-5.7781],[14,-5.4869],[15,-5.3218],[16,-5.0399],[17,-4.8283],[18,-4.6399],[19,-4.3083],[20,-4.1307],[21,-3.8429],[22,-3.5232],[23,-3.2215],[24,-3.0925],[25,-3.0466],[26,-3.2591],[27,-3.5659],[28,-3.9365],[29,-4.3628],[30,-4.9347],[31,-5.7347],[32,-6.2933],[33,-6.5999],[34,-6.6595],[35,-6.685],[36,-6.5074],[37,-6.3788],[38,-6.3121],[39,-6.2055],[41,-5.5614],[42,-5.2018],[43,-4.9205],[45,-4.7716],[46,-4.1958],[47,-3.4134],[48,-2.3443],[49,-1.231],[50,-0.1987],[51,0.3541],[52,1.1496],[53,1.9298],[54,1.8762],[55,0.8732],[56,0.1232],[57,-0.9121],[58,-1.4372],[59,-1.5616],[60,-1.8782],[61,-2.9468],[62,-3.5898],[63,-4.3302],[64,-4.7989],[65,-5.0397],[67,-4.8422],[68,-5.1224],[69,-5.1702],[70,-4.9402],[71,-5.0254],[72,-4.7295],[73,-4.5192],[74,-4.5187],[75,-4.4718],[76,-4.4478],[77,-4.577],[78,-4.796],[79,-4.9993],[80,-5.1128],[81,-5.2099],[82,-5.2714],[84,-5.3563],[85,-5.4103],[86,-5.4955],[87,-5.6334],[89,-5.8385],[90,-5.9566],[92,-6.1469],[95,-6.3851],[98,-6.6513],[99,-6.7217],[100,-6.8379],[102,-6.9594],[103,-6.9994],[104,-7.0742],[108,-7.1558],[111,-7.2749],[113,-7.3146],[114,-7.3229],[117,-7.2231],[118,-7.2011],[119,-7.203],[122,-7.2413],[125,-7.2303],[128,-7.2735],[131,-7.2686],[133,-7.3113],[134,-7.3884],[137,-7.5227],[139,-7.5671],[144,-7.6444],[146,-7.7168],[147,-7.7662],[148,-7.7845],[149,-7.8223],[151,-7.8331],[152,-7.8104],[153,-7.7396],[154,-7.6146],[158,-7.2433],[160,-7.0934],[161,-7.0632],[162,-7.0177],[163,-7.0011],[165,-7.0488],[166,-7.0614],[167,-7.0292],[168,-6.8052],[169,-6.7771],[170,-6.8853],[171,-6.8496],[172,-6.9536],[173,-5.5317],[174,-1.4091],[175,-6.1877],[176,-0.9621],[177,-0.554],[178,-4.474],[179,-0.7419],[180,-3.0284],[181,-1.4377],[182,3.5364],[183,1.3034],[184,2.2006]]
       }
       leg_l_pv_ik_ctrl.Rotation.Z {
        type=MovieSceneScriptingFloatChannel
        default=-14.0921
        thinned_from=185
        keys [[0,4.6657],[1,4.6529],[2,4.6947],[3,4.6797],[4,4.8114],[5,5.0919],[6,5.5073],[7,5.9553],[8,6.35],[9,6.6001],[10,7.0863],[12,7.9545],[13,8.1041],[14,8.0358],[15,7.785],[17,7.1691],[18,6.7836],[19,6.549],[20,6.4178],[21,6.0058],[22,5.8336],[23,5.7473],[24,5.589],[25,5.4029],[26,5.2851],[27,5.5126],[28,5.8074],[29,5.7998],[30,5.8283],[31,6.1102],[33,6.9341],[34,7.0062],[35,6.9531],[36,7.0583],[37,7.1181],[39,6.8435],[40,6.832],[41,6.6634],[42,6.3625],[43,6.2549],[45,6.4051],[46,6.7552],[47,6.9774],[48,7.0564],[49,6.961],[50,6.7393],[51,6.6143],[52,6.3964],[53,5.3526],[54,4.6059],[55,3.6922],[56,3.5057],[57,3.3932],[58,3.5325],[59,3.3787],[60,3.242],[61,3.4598],[62,3.958],[63,4.4682],[64,5.1043],[65,5.6907],[66,5.8523],[67,5.6428],[68,5.2579],[69,4.737],[70,3.8617],[71,3.6189],[72,3.0742],[73,3.1101],[74,3.5024],[75,3.7855],[76,4.1135],[77,4.4809],[79,5.0617],[80,5.3142],[81,5.5333],[82,5.4969],[83,5.3949],[85,5.8473],[86,5.9076],[87,6.2169],[88,6.6723],[89,7.5848],[90,8.3537],[91,8.8246],[92,9.0352],[93,9.0976],[94,9.1346],[95,9.1052],[96,8.7874],[97,8.1543],[98,7.621],[99,7.414],[100,7.0274],[101,6.8567],[102,6.7697],[103,6.7763],[104,6.6819],[105,6.6313],[106,6.6579],[107,6.7947],[108,6.9512],[109,6.796],[110,6.617],[111,6.4841],[112,6.4505],[113,6.2505],[114,5.9734],[115,5.8751],[116,5.8871],[117,5.9435],[118,5.8701],[119,5.7093],[120,5.6028],[121,5.5402],[122,5.4102],[123,5.1701],[124,5.0893],[125,5.0845],[126,4.9989],[127,4.8875],[128,4.6711],[129,4.638],[131,4.7919],[132,4.8069],[133,4.748],[134,4.5921],[136,4.7104],[137,4.6433],[138,4.5567],[140,4.3452],[141,4.3179],[142,4.3776],[143,4.503],[144,4.5428],[145,4.5374],[146,4.5848],[147,4.5452],[148,4.4921],[149,4.4217],[150,4.3169],[152,3.9898],[153,3.7438],[154,3.4335],[155,3.2386],[156,3.0714],[157,3.0348],[158,2.8764],[159,2.6595],[160,2.3004],[161,2.1998],[162,1.8498],[163,1.6321],[164,1.7109],[165,2.0753],[166,2.2044],[167,2.371],[168,2.8298],[169,2.4088],[170,2.386],[171,0.8183],[172,-1.1243],[173,4.6345],[174,-8.1094],[175,3.5738],[176,-10.8703],[177,-10.4932],[178,3.8509],[179,-9.9514],[180,6.1492],[181,4.5037],[182,-1.6813],[183,-18.7585],[184,-14.0921]]
       }
       leg_l_pv_ik_ctrl.Scale.X {type=MovieSceneScriptingFloatChannel, default=1, const [1,185]}
       leg_l_pv_ik_ctrl.Scale.Y {type=MovieSceneScriptingFloatChannel, default=1, const [1,185]}
       leg_l_pv_ik_ctrl.Scale.Z {type=MovieSceneScriptingFloatChannel, default=1, const [1,185]}
       thigh_r_fk_ctrl.Location.X {type=MovieSceneScriptingFloatChannel, default=-0.0002, thinned_from=128, keys [[0,-0.0002],[184,0]]}
       thigh_r_fk_ctrl.Location.Y {type=MovieSceneScriptingFloatChannel, default=0, thinned_from=128, keys [[0,0],[184,0]]}
       thigh_r_fk_ctrl.Location.Z {type=MovieSceneScriptingFloatChannel, default=0.0001, thinned_from=128, keys [[0,0],[184,0]]}
       thigh_r_fk_ctrl.Rotation.X {
        type=MovieSceneScriptingFloatChannel
        default=10.5627
        thinned_from=128
        keys [[0,-11.8544],[2,-11.8703],[4,-11.8576],[5,-11.9049],[6,-11.8536],[7,-11.7224],[8,-11.4664],[9,-11.3136],[10,-11.3147],[11,-11.281],[13,-11.1137],[14,-10.9902],[15,-10.8906],[16,-10.9057],[17,-10.9645],[18,-10.9843],[19,-11.035],[20,-11.0495],[22,-11.1797],[24,-11.2277],[27,-11.4791],[28,-11.5234],[29,-11.5223],[31,-11.6907],[32,-11.7093],[33,-11.5725],[34,-11.4752],[37,-11.0612],[38,-11.0137],[39,-10.9943],[40,-10.9246],[41,-10.8976],[42,-10.9006],[45,-10.7743],[46,-10.6054],[47,-10.5023],[48,-10.5453],[49,-10.7639],[50,-11.1916],[51,-11.1726],[52,-10.8446],[53,-10.7279],[54,-9.999],[55,-9.6447],[56,-8.5404],[57,-7.5659],[58,-6.4655],[102,15.151],[103,15.0501],[106,15.2037],[108,15.2096],[109,15.1841],[110,15.2452],[111,15.3894],[112,15.4936],[113,15.5535],[114,15.5607],[115,15.5262],[116,15.5895],[117,15.7017],[118,15.7776],[121,15.7278],[123,15.7527],[125,15.7002],[126,15.7714],[128,15.8337],[130,15.7982],[132,15.635],[133,15.6236],[134,15.5713],[136,15.2703],[137,15.214],[138,15.2418],[141,15.4837],[142,15.4855],[144,15.2507],[145,15.0948],[146,15.0007],[147,14.9742],[148,14.9871],[150,15.0949],[151,15.1779],[152,15.3034],[153,15.474],[154,15.7555],[156,16.157],[158,16.3084],[161,16.3217],[162,16.2904],[163,16.1767],[164,15.9245],[165,15.6236],[166,15.2296],[172,13.7852],[177,14.8599],[181,6.2051],[182,4.2036],[183,1.4222],[184,0]]
       }
       thigh_r_fk_ctrl.Rotation.Y {
        type=MovieSceneScriptingFloatChannel
        default=-5.005
        thinned_from=128
        keys [[0,-2.7483],[1,-2.6443],[2,-2.5788],[3,-2.4515],[4,-2.4643],[5,-2.5481],[6,-2.6172],[7,-2.5953],[8,-2.4278],[9,-2.2679],[10,-2.3108],[11,-2.3994],[12,-2.4615],[13,-2.3973],[14,-2.2054],[15,-2.0703],[16,-2.0258],[17,-2.0244],[18,-2.0294],[19,-2.0922],[20,-2.0982],[21,-2.0171],[23,-1.991],[24,-1.9151],[25,-1.8129],[26,-1.7854],[27,-1.9714],[28,-2.0838],[29,-1.9739],[30,-1.9442],[31,-2.0801],[32,-2.2538],[33,-2.3056],[34,-2.2292],[35,-1.9917],[36,-1.9091],[37,-1.915],[38,-1.8824],[39,-1.9063],[40,-2.0562],[41,-2.1005],[42,-2.059],[43,-2.1345],[44,-2.2221],[45,-2.2666],[46,-2.4118],[47,-2.4819],[48,-2.6077],[49,-2.8177],[50,-3.1853],[51,-3.408],[52,-3.7991],[53,-4.1698],[54,-4.1809],[55,-4.5779],[57,-4.7954],[58,-4.7214],[102,0.2283],[103,0.2237],[106,0.2537],[107,0.2464],[108,0.178],[109,0.1763],[110,0.2663],[111,0.4295],[112,0.6507],[113,0.9539],[114,1.2353],[115,1.5975],[116,1.6725],[119,1.9272],[120,1.9608],[122,1.9638],[123,1.9049],[124,1.8101],[125,1.7325],[126,1.7205],[127,1.7768],[128,1.8722],[129,1.8983],[130,1.8145],[131,1.7459],[132,1.7521],[133,1.7857],[134,1.8996],[135,1.9504],[136,1.9025],[137,1.8838],[138,1.9113],[139,1.9475],[140,2.0119],[141,2.0892],[142,2.1497],[143,2.1061],[145,2.0692],[146,2.1012],[147,2.1414],[148,2.1501],[149,2.1807],[150,2.1867],[152,2.1806],[153,2.17],[155,2.2672],[157,2.4144],[158,2.5519],[159,2.7344],[160,2.9448],[161,3.0641],[162,3.3617],[163,3.7045],[164,4.03],[165,4.2017],[166,4.1476],[172,-3.1768],[177,-3.1655],[181,0.7217],[182,1.6133],[183,-8.4216],[184,-8.0673]]
       }
       thigh_r_fk_ctrl.Rotation.Z {
        type=MovieSceneScriptingFloatChannel
        default=-9.8175
        thinned_from=128
        keys [[0,-0.7822],[1,-0.8054],[2,-0.8817],[3,-0.9502],[4,-0.9831],[5,-0.996],[6,-0.9451],[7,-0.9032],[8,-0.8951],[9,-0.8726],[10,-0.8064],[11,-0.6944],[12,-0.5233],[13,-0.3846],[14,-0.2277],[15,-0.1641],[16,-0.1398],[17,-0.113],[18,-0.054],[19,0.0118],[20,0.0526],[21,0.1171],[22,0.1793],[23,0.2189],[24,0.2482],[25,0.2585],[26,0.2814],[27,0.3775],[28,0.5214],[29,0.576],[30,0.6091],[31,0.6638],[32,0.7505],[33,0.8943],[34,0.9367],[35,1.0551],[36,1.147],[37,1.1572],[38,1.1705],[39,1.1933],[40,1.2428],[41,1.3174],[42,1.3701],[43,1.3896],[45,1.4884],[46,1.6336],[47,1.8365],[48,2.2321],[49,2.7008],[50,3.5345],[51,3.5878],[52,3.3609],[53,3.9193],[54,2.7894],[55,2.5849],[56,0.8016],[57,-0.8916],[58,-2.8836],[102,-44.021],[103,-44.0199],[107,-43.7554],[111,-43.7785],[114,-43.499],[117,-43.0439],[125,-42.8256],[127,-42.9239],[133,-42.7315],[136,-42.8788],[140,-42.5954],[156,-42.9198],[159,-42.6288],[162,-42.2728],[165,-42.1427],[166,-42.2198],[172,-50.9379],[177,-50.947],[181,-38.4483],[182,-25.4988],[183,-14.8971],[184,0]]
       }
       thigh_r_fk_ctrl.Scale.X {type=MovieSceneScriptingFloatChannel, default=1, const [1,128]}
       thigh_r_fk_ctrl.Scale.Y {type=MovieSceneScriptingFloatChannel, default=1, const [1,128]}
       thigh_r_fk_ctrl.Scale.Z {type=MovieSceneScriptingFloatChannel, default=1, const [1,128]}
       calf_r_fk_ctrl.Location.X {
        type=MovieSceneScriptingFloatChannel
        default=0.0003
        thinned_from=127
        keys [[0,0.0002],[5,0.0002],[6,0.0003],[11,0.0002],[26,0.0002],[47,0.0002],[49,0.0003],[51,0.0002],[54,0.0006],[55,-0.0002],[56,0.0005],[57,0.0008],[58,0.0004],[111,0.0002],[112,0.0004],[113,0.0002],[177,0.0003],[184,0]]
       }
       calf_r_fk_ctrl.Location.Y {
        type=MovieSceneScriptingFloatChannel
        default=0
        thinned_from=127
        keys [[0,0.0004],[1,-0.0001],[2,0.0011],[3,0.0009],[4,-0.0006],[5,-0.0005],[6,0.0007],[7,-0.0002],[9,0.0003],[10,-0.0005],[11,-0.0004],[12,0.0011],[13,0.0006],[14,0.0012],[15,0.0003],[16,0.0007],[17,0.0004],[18,-0.0004],[19,0.0002],[20,0.0011],[21,0.0003],[22,0.0003],[23,-0.0001],[24,-0.0001],[25,0.0006],[26,-0.0002],[27,0.0009],[28,0.0002],[29,-0.0002],[30,-0.0002],[31,0.0008],[32,0.0005],[33,0.0006],[34,0],[35,-0.0002],[36,0],[37,-0.0004],[39,0.001],[40,-0.0003],[41,-0.0006],[42,-0.0004],[43,0.0008],[44,0.0009],[45,0.0005],[46,-0.0006],[47,-0.0001],[48,0.001],[49,0.001],[50,-0.0009],[51,-0.0006],[52,0.0011],[53,0.0015],[54,0.0008],[55,-0.0104],[56,0.0005],[57,0],[102,0.0008],[106,-0.001],[107,0.0006],[108,0.0016],[109,0.0016],[110,0.0006],[111,0],[112,0.002],[113,-0.0006],[114,0.0005],[115,0.0003],[116,-0.0009],[117,-0.0007],[118,0.0004],[119,0.0003],[120,-0.0004],[121,0.0008],[122,0.0001],[123,0.0003],[125,-0.0008],[126,0.0003],[127,0.0009],[128,-0.0003],[129,0.0009],[130,0.0001],[131,0.0006],[132,-0.0004],[133,-0.0005],[134,-0.0001],[135,0.0008],[136,0.0011],[137,0.0011],[138,-0.0001],[139,-0.0001],[140,-0.0008],[141,0.0004],[145,0.0001],[146,0.0008],[147,-0.0001],[148,-0.0002],[149,0.0006],[150,0.0006],[151,-0.0002],[152,0.0002],[153,-0.0001],[154,0.0002],[155,-0.0006],[156,-0.0009],[157,0.0001],[158,-0.0004],[160,-0.0004],[161,-0.0001],[162,-0.0007],[165,0],[177,-0.0002],[184,0]]
       }
       calf_r_fk_ctrl.Location.Z {
        type=MovieSceneScriptingFloatChannel
        default=0.0001
        thinned_from=127
        keys [[0,0.0001],[1,0],[3,0.0001],[4,-0.0001],[5,-0.0001],[6,0.0001],[11,-0.0001],[12,0.0002],[19,0],[39,0.0001],[42,0],[45,0],[47,0],[49,0.0001],[50,-0.0001],[52,0],[54,-0.0003],[55,0.0021],[56,-0.0004],[58,-0.0001],[102,0.0004],[106,-0.0005],[107,0.0004],[108,0.0008],[109,0.0008],[110,0.0003],[111,0],[112,0.0011],[113,-0.0003],[114,0.0003],[115,0.0002],[116,-0.0005],[117,-0.0004],[118,0.0003],[119,0.0002],[120,-0.0002],[121,0.0005],[122,0.0001],[123,0.0002],[125,-0.0005],[126,0.0002],[127,0.0006],[128,-0.0002],[129,0.0006],[130,0.0001],[131,0.0004],[132,-0.0002],[133,-0.0003],[134,-0.0001],[135,0.0005],[137,0.0007],[138,0],[139,0],[140,-0.0005],[141,0.0003],[145,0.0001],[146,0.0006],[147,0],[148,-0.0001],[149,0.0004],[150,0.0004],[151,-0.0001],[152,0.0001],[153,-0.0001],[154,0.0001],[155,-0.0004],[156,-0.0006],[157,0.0001],[158,-0.0003],[160,-0.0003],[161,0],[162,-0.0005],[163,-0.0002],[166,0.0001],[184,0]]
       }
       calf_r_fk_ctrl.Rotation.X {
        type=MovieSceneScriptingFloatChannel
        default=0.2066
        thinned_from=127
        keys [[0,-7.8062],[2,-7.8119],[5,-7.794],[12,-7.8109],[26,-7.7931],[28,-7.8258],[34,-7.8335],[36,-7.8134],[47,-7.8195],[49,-7.7959],[50,-7.798],[51,-7.7189],[52,-7.4937],[53,-7.1526],[54,-6.7233],[55,-6.2374],[57,-5.0883],[58,-4.5265],[102,-0.8667],[103,-0.8617],[107,-0.8668],[110,-0.86],[111,-0.8533],[112,-0.8437],[113,-0.8473],[114,-0.8651],[115,-0.8796],[116,-0.8725],[117,-0.8688],[119,-0.8759],[121,-0.8765],[122,-0.8743],[124,-0.885],[126,-0.8856],[127,-0.884],[128,-0.8794],[129,-0.8802],[130,-0.8877],[131,-0.8897],[132,-0.8838],[133,-0.8845],[135,-0.8961],[137,-0.8896],[138,-0.8932],[139,-0.9023],[140,-0.9051],[141,-0.9057],[142,-0.8957],[143,-0.8974],[145,-0.909],[147,-0.9046],[149,-0.9061],[150,-0.9051],[151,-0.8994],[152,-0.8976],[153,-0.9017],[155,-0.8872],[156,-0.8772],[157,-0.872],[158,-0.8735],[159,-0.8839],[160,-0.899],[161,-0.907],[162,-0.9236],[163,-0.9425],[164,-0.9699],[165,-0.9931],[172,-1.1917],[177,-1.1634],[181,1.1582],[182,1.1995],[184,0]]
       }
       calf_r_fk_ctrl.Rotation.Y {
        type=MovieSceneScriptingFloatChannel
        default=0.4093
        thinned_from=127
        keys [[0,-0.1036],[1,-0.1029],[2,-0.1049],[3,-0.0987],[4,-0.094],[5,-0.0982],[6,-0.1106],[7,-0.1188],[8,-0.1176],[9,-0.1128],[10,-0.1181],[11,-0.1273],[12,-0.1352],[13,-0.1373],[14,-0.1346],[15,-0.133],[16,-0.1304],[17,-0.1304],[18,-0.1311],[19,-0.1324],[20,-0.1322],[21,-0.1315],[22,-0.1345],[23,-0.1352],[24,-0.131],[25,-0.1257],[26,-0.1296],[27,-0.1468],[28,-0.1671],[29,-0.1702],[30,-0.1718],[31,-0.1814],[32,-0.1982],[33,-0.2116],[34,-0.2112],[35,-0.2052],[36,-0.2061],[37,-0.208],[38,-0.202],[39,-0.1984],[40,-0.204],[41,-0.21],[42,-0.2127],[43,-0.2117],[44,-0.2079],[45,-0.205],[46,-0.1962],[47,-0.1803],[48,-0.1538],[49,-0.1347],[50,-0.1379],[51,-0.0527],[52,0.1737],[53,0.4849],[54,0.9168],[55,1.3952],[56,1.9737],[57,2.5857],[58,3.1698],[102,7.315],[103,7.3192],[112,7.2915],[115,7.2144],[137,7.1917],[141,7.1553],[157,7.164],[159,7.1219],[160,7.0854],[161,7.0657],[164,6.9288],[166,6.8464],[172,7.394],[177,7.4914],[181,-0.312],[182,-0.0097],[184,0]]
       }
       calf_r_fk_ctrl.Rotation.Z {
        type=MovieSceneScriptingFloatChannel
        default=5.7705
        thinned_from=127
        keys [[0,17.3368],[1,17.3448],[2,17.3155],[4,17.4158],[5,17.3847],[6,17.2751],[7,17.2042],[9,17.2435],[10,17.2079],[12,17.0718],[13,17.0543],[21,17.1354],[23,17.1124],[25,17.1884],[26,17.1535],[28,16.8244],[30,16.7814],[31,16.6998],[32,16.5542],[33,16.4417],[34,16.4471],[35,16.502],[37,16.4876],[39,16.5851],[42,16.4773],[43,16.4881],[45,16.573],[46,16.6803],[47,16.8469],[48,17.1217],[49,17.3248],[50,17.3593],[51,18.4518],[52,21.3531],[53,25.4952],[54,30.9846],[55,37.2097],[56,44.3736],[57,51.8889],[58,58.9926],[102,100.9123],[112,100.7138],[115,99.9803],[138,99.6823],[157,99.4787],[159,99.0552],[161,98.4864],[166,96.2542],[172,100.7486],[177,101.6745],[181,102.6564],[182,88.0463],[184,0]]
       }
       calf_r_fk_ctrl.Scale.X {type=MovieSceneScriptingFloatChannel, default=1, const [1,127]}
       calf_r_fk_ctrl.Scale.Y {type=MovieSceneScriptingFloatChannel, default=1, const [1,127]}
       calf_r_fk_ctrl.Scale.Z {type=MovieSceneScriptingFloatChannel, default=1, const [1,127]}
       foot_r_fk_ctrl.Location.X {
        type=MovieSceneScriptingFloatChannel
        default=0.0003
        thinned_from=129
        keys [[0,0.0001],[54,0.0004],[55,0.0008],[56,0.0006],[57,0.0011],[58,0.0005],[165,0.0001],[166,0.0009],[172,0.0008],[177,0.0009],[181,0],[184,0]]
       }
       foot_r_fk_ctrl.Location.Y {type=MovieSceneScriptingFloatChannel, default=0.0001, thinned_from=129, keys [[0,0],[177,0],[181,-0.0329],[182,-0.0329],[184,0]]}
       foot_r_fk_ctrl.Location.Z {type=MovieSceneScriptingFloatChannel, default=0, thinned_from=129, keys [[0,0],[177,0],[181,-0.9075],[182,-0.9075],[184,0]]}
       foot_r_fk_ctrl.Rotation.X {
        type=MovieSceneScriptingFloatChannel
        default=-1.2092
        thinned_from=129
        keys [[0,14.3352],[1,14.1114],[2,14.0768],[4,14.2392],[5,14.2683],[6,14.1817],[7,14.1953],[8,14.2502],[9,14.3391],[11,14.3628],[12,14.2878],[13,14.3179],[14,14.2339],[15,14.2368],[16,14.3053],[17,14.2686],[19,14.3511],[20,14.3171],[21,14.1959],[22,14.1655],[23,14.0785],[24,14.0363],[25,14.1217],[27,14.0748],[28,13.978],[29,13.986],[30,13.9012],[31,13.9111],[32,13.8734],[33,13.763],[34,13.7436],[36,13.7809],[37,13.6776],[38,13.6826],[39,13.7592],[40,13.8789],[41,13.848],[42,13.7794],[43,13.8677],[45,13.8112],[46,13.8856],[48,13.7381],[49,13.9848],[51,13.9006],[52,13.9076],[53,14.0492],[54,13.6031],[55,13.3253],[56,13.8162],[57,14.6743],[58,15.1064],[102,13.1293],[105,13.109],[106,13.1935],[107,13.2289],[108,13.1671],[109,13.0765],[111,12.8307],[112,12.675],[114,11.9161],[115,11.3013],[116,11.169],[117,11.1091],[118,11.0067],[119,10.796],[120,10.7102],[122,10.7718],[123,10.7506],[126,10.752],[127,10.7162],[128,10.6248],[130,10.4967],[131,10.502],[133,10.3705],[134,10.1916],[135,10.1474],[137,10.1913],[138,10.1227],[139,10.1165],[141,10.0565],[142,10.1065],[143,10.1022],[144,9.975],[145,9.8939],[148,9.9672],[149,10.0341],[151,10.1176],[152,10.1188],[154,10.0393],[155,9.9726],[156,9.9321],[157,9.855],[158,9.7488],[159,9.7036],[160,9.6142],[161,9.5598],[162,9.4132],[163,9.3067],[164,9.1226],[165,9.0093],[166,9.1074],[172,19.8457],[177,21.4442],[181,-9.8979],[182,-7.4984],[184,9.4952]]
       }
       foot_r_fk_ctrl.Rotation.Y {
        type=MovieSceneScriptingFloatChannel
        default=-16.8654
        thinned_from=129
        keys [[0,-0.92],[1,-0.9833],[2,-0.9069],[3,-0.9598],[4,-0.9635],[5,-1.1475],[6,-1.2733],[7,-1.2986],[8,-1.22],[9,-1.0674],[10,-1.2672],[11,-1.3535],[12,-1.3776],[13,-1.4662],[14,-1.4818],[15,-1.3951],[16,-1.3946],[17,-1.4219],[18,-1.3989],[19,-1.4293],[21,-1.4482],[22,-1.5515],[23,-1.5824],[24,-1.5201],[25,-1.5693],[26,-1.5873],[27,-1.6552],[28,-1.7339],[29,-1.6706],[30,-1.7499],[31,-1.8843],[32,-2.0637],[33,-2.3285],[34,-2.3022],[35,-2.2348],[36,-2.398],[37,-2.3421],[38,-2.2747],[40,-2.396],[41,-2.4107],[42,-2.4765],[43,-2.4976],[44,-2.4872],[45,-2.458],[46,-2.3478],[47,-2.1608],[48,-2.0425],[49,-1.6826],[50,-1.7002],[51,-1.5907],[52,-1.5293],[53,-1.7678],[54,-1.6735],[55,-1.7333],[56,-1.4525],[57,-0.7171],[58,-0.2135],[102,4.255],[103,4.1919],[104,4.056],[105,4.0356],[106,3.975],[107,3.9641],[108,3.8178],[109,3.7832],[110,3.898],[111,3.8717],[112,3.5137],[113,3.2612],[114,3.0974],[115,2.7846],[116,2.5312],[117,2.4811],[118,2.5831],[119,2.5881],[120,2.5241],[121,2.485],[122,2.4353],[123,2.4972],[124,2.5434],[125,2.5045],[126,2.588],[128,2.7029],[129,2.6446],[130,2.6647],[131,2.61],[132,2.4147],[133,2.422],[134,2.5003],[135,2.6396],[136,2.5591],[137,2.4024],[138,2.2881],[139,2.2586],[140,2.182],[141,2.1135],[142,1.9761],[143,1.9742],[144,2.071],[145,2.1347],[146,2.1159],[148,2.0588],[149,2.0151],[150,1.9437],[151,1.8076],[152,1.734],[153,1.9016],[154,1.975],[155,2.1397],[156,2.1962],[157,2.0543],[158,1.8756],[159,1.7808],[160,1.5383],[161,1.4189],[162,1.0563],[163,0.8422],[164,0.7205],[165,0.6551],[166,0.5801],[172,-4.794],[177,-4.8442],[181,29.9993],[182,20.9841],[184,6.4715]]
       }
       foot_r_fk_ctrl.Rotation.Z {
        type=MovieSceneScriptingFloatChannel
        default=31.0622
        thinned_from=129
        keys [[0,32.4065],[1,32.315],[4,32.3042],[7,32.1274],[8,32.1632],[9,32.2814],[11,32.1294],[12,32.1138],[13,32.1916],[15,32.1221],[17,32.1497],[19,32.0258],[21,32.004],[23,31.8574],[25,31.945],[28,31.8805],[29,31.9526],[33,31.8579],[36,31.9415],[38,31.8923],[42,31.6097],[46,31.3817],[48,31.4206],[49,31.6021],[50,31.5889],[51,31.3977],[52,30.8845],[53,30.0362],[54,28.7175],[55,27.0014],[57,23.2608],[58,21.7088],[102,-0.223],[103,-0.2448],[104,-0.3624],[105,-0.3989],[106,-0.3517],[107,-0.3308],[108,-0.3657],[109,-0.2953],[110,-0.1783],[111,-0.1399],[112,-0.1983],[113,-0.34],[114,-0.4759],[115,-0.5718],[116,-0.6851],[117,-0.773],[118,-0.7694],[119,-0.8301],[120,-0.9089],[121,-0.9433],[122,-0.9544],[123,-0.9591],[124,-0.9386],[125,-0.9388],[126,-0.8384],[127,-0.7896],[128,-0.6841],[129,-0.6535],[130,-0.657],[131,-0.6321],[132,-0.665],[133,-0.6492],[134,-0.5967],[135,-0.5751],[136,-0.6155],[137,-0.6733],[138,-0.7223],[140,-0.7465],[141,-0.7667],[142,-0.8116],[143,-0.836],[144,-0.7653],[145,-0.7154],[146,-0.6953],[147,-0.6883],[148,-0.728],[149,-0.7821],[150,-0.8139],[151,-0.8389],[152,-0.851],[153,-0.7345],[154,-0.6292],[155,-0.4849],[156,-0.3156],[157,-0.2196],[158,-0.0985],[159,0.0326],[160,0.1395],[161,0.2602],[162,0.4339],[163,0.6148],[164,0.6957],[165,0.7378],[166,0.9629],[172,0.7615],[177,0.6149],[181,-5.456],[182,-3.6698],[184,-0.7959]]
       }
       foot_r_fk_ctrl.Scale.X {type=MovieSceneScriptingFloatChannel, default=1, const [1,129]}
       foot_r_fk_ctrl.Scale.Y {type=MovieSceneScriptingFloatChannel, default=1, const [1,129]}
       foot_r_fk_ctrl.Scale.Z {type=MovieSceneScriptingFloatChannel, default=1, const [1,129]}
       ball_r_fk_ctrl.Location.X {
        type=MovieSceneScriptingFloatChannel
        default=0
        thinned_from=178
        keys [[0,0],[56,0.0001],[57,0.0002],[167,0],[169,0.0001],[170,-0.0001],[171,0.0001],[172,0],[177,0],[181,-0.0003],[182,0.0013],[183,0.0051],[184,0]]
       }
       ball_r_fk_ctrl.Location.Y {
        type=MovieSceneScriptingFloatChannel
        default=-0.0001
        thinned_from=178
        keys [[0,0],[56,-0.0001],[57,-0.0003],[58,-0.0001],[59,-0.0003],[60,0],[167,0],[168,-0.0003],[170,-0.0004],[172,0],[177,-0.0001],[181,-0.0009],[182,-0.0075],[183,-0.0064],[184,-0.0001]]
       }
       ball_r_fk_ctrl.Location.Z {
        type=MovieSceneScriptingFloatChannel
        default=0
        thinned_from=178
        keys [[0,0],[56,0],[167,0],[170,0.0002],[172,0],[181,0.0002],[182,-0.0024],[183,0.0027],[184,0]]
       }
       ball_r_fk_ctrl.Rotation.X {
        type=MovieSceneScriptingFloatChannel
        default=-0.0101
        thinned_from=178
        keys [[0,-0.0556],[1,-0.0555],[2,-0.0564],[3,-0.0554],[4,-0.0553],[5,-0.0555],[6,-0.0553],[7,-0.0549],[9,-0.0556],[10,-0.0548],[11,-0.0551],[12,-0.0543],[13,-0.0555],[14,-0.0549],[15,-0.0544],[17,-0.055],[19,-0.0547],[20,-0.0552],[21,-0.0555],[23,-0.0537],[24,-0.0542],[25,-0.053],[26,-0.0528],[27,-0.0533],[28,-0.053],[29,-0.0536],[30,-0.0526],[31,-0.0518],[32,-0.0522],[33,-0.0514],[34,-0.0515],[35,-0.0521],[36,-0.0522],[37,-0.0518],[38,-0.0526],[39,-0.0517],[40,-0.0512],[43,-0.0508],[44,-0.051],[45,-0.0505],[46,-0.0501],[47,-0.0501],[48,-0.0506],[49,-0.0504],[50,-0.0489],[51,-0.051],[52,-0.0514],[53,-0.0528],[54,-0.0534],[55,-0.0551],[56,-0.053],[57,-0.0535],[58,-0.0535],[60,-0.0532],[61,-0.0539],[62,-0.0541],[63,-0.0539],[64,-0.0548],[65,-0.0539],[66,-0.0559],[67,-0.0534],[68,-0.046],[69,-0.0425],[70,-0.0462],[71,-0.0516],[72,-0.0554],[74,-0.0576],[75,-0.0571],[76,-0.0559],[77,-0.0562],[78,-0.0569],[79,-0.0566],[80,-0.0549],[81,-0.0553],[82,-0.0553],[83,-0.0533],[84,-0.0536],[85,-0.0545],[86,-0.0541],[87,-0.0547],[88,-0.0549],[89,-0.0544],[90,-0.0535],[91,-0.0531],[92,-0.0518],[93,-0.052],[94,-0.0514],[95,-0.0515],[96,-0.0524],[97,-0.0524],[98,-0.0512],[99,-0.052],[100,-0.0512],[101,-0.0524],[102,-0.0543],[103,-0.0547],[104,-0.0522],[105,-0.052],[106,-0.0526],[107,-0.0538],[108,-0.0528],[109,-0.0534],[110,-0.053],[111,-0.0543],[112,-0.0536],[113,-0.0534],[114,-0.0543],[115,-0.055],[117,-0.0551],[118,-0.0547],[120,-0.0549],[122,-0.0546],[124,-0.0538],[125,-0.0538],[126,-0.0549],[128,-0.0562],[129,-0.056],[130,-0.056],[131,-0.0563],[132,-0.055],[133,-0.055],[134,-0.0563],[135,-0.0565],[136,-0.0566],[137,-0.0563],[138,-0.0554],[139,-0.0555],[140,-0.0559],[141,-0.0558],[142,-0.054],[143,-0.0537],[145,-0.0561],[147,-0.0562],[148,-0.056],[149,-0.0548],[150,-0.0546],[151,-0.0536],[152,-0.0529],[153,-0.0543],[154,-0.0545],[155,-0.0554],[156,-0.0567],[157,-0.0564],[159,-0.0566],[160,-0.0562],[161,-0.0567],[164,-0.0573],[165,-0.057],[166,-0.057],[167,-0.0577],[168,-0.0565],[169,-0.0564],[170,-0.0586],[171,-0.0561],[172,-0.056],[177,-0.0543],[181,-0.0499],[182,-0.0551],[183,-0.0633],[184,-0.0101]]
       }
       ball_r_fk_ctrl.Rotation.Y {
        type=MovieSceneScriptingFloatChannel
        default=0.0103
        thinned_from=178
        keys [[0,-0.2353],[1,-0.2345],[2,-0.24],[3,-0.2337],[4,-0.2332],[5,-0.2342],[6,-0.2333],[7,-0.2307],[8,-0.2327],[9,-0.2353],[10,-0.2303],[11,-0.232],[12,-0.2273],[13,-0.2345],[14,-0.2305],[15,-0.2279],[16,-0.2291],[17,-0.2311],[19,-0.2296],[20,-0.2329],[21,-0.2343],[22,-0.2285],[23,-0.2234],[24,-0.2265],[25,-0.2191],[26,-0.2182],[27,-0.2212],[28,-0.219],[29,-0.2227],[30,-0.2168],[31,-0.2122],[32,-0.2144],[33,-0.2094],[34,-0.2102],[35,-0.2137],[36,-0.2147],[37,-0.2123],[38,-0.2171],[39,-0.2116],[40,-0.2085],[41,-0.2074],[42,-0.2072],[43,-0.2058],[44,-0.2076],[45,-0.2041],[46,-0.202],[47,-0.2022],[48,-0.2048],[49,-0.2036],[50,-0.1947],[51,-0.2076],[52,-0.2094],[53,-0.2179],[54,-0.222],[55,-0.2323],[56,-0.2194],[57,-0.2222],[58,-0.2227],[60,-0.2205],[61,-0.2247],[62,-0.2256],[63,-0.2245],[64,-0.2301],[65,-0.2248],[66,-0.2368],[67,-0.222],[68,-0.1778],[69,-0.1581],[70,-0.1792],[71,-0.2104],[72,-0.234],[73,-0.2412],[74,-0.2476],[75,-0.2444],[76,-0.2368],[77,-0.239],[78,-0.2432],[79,-0.241],[80,-0.231],[81,-0.2329],[82,-0.2334],[83,-0.221],[84,-0.2228],[85,-0.2283],[86,-0.2262],[87,-0.2293],[88,-0.2311],[89,-0.2278],[90,-0.2222],[91,-0.2201],[92,-0.2121],[93,-0.2131],[94,-0.2098],[95,-0.2102],[96,-0.2156],[97,-0.2156],[98,-0.2082],[99,-0.2135],[100,-0.2087],[101,-0.2157],[102,-0.227],[103,-0.2296],[104,-0.2147],[105,-0.2132],[106,-0.2168],[107,-0.2241],[108,-0.2181],[109,-0.2216],[110,-0.219],[111,-0.2271],[112,-0.2225],[113,-0.2217],[114,-0.2273],[115,-0.2313],[116,-0.2319],[117,-0.2318],[118,-0.2298],[120,-0.2307],[122,-0.2286],[123,-0.2259],[124,-0.2242],[125,-0.2242],[126,-0.2306],[128,-0.239],[129,-0.2373],[130,-0.2377],[131,-0.2396],[132,-0.2315],[133,-0.2313],[134,-0.2391],[135,-0.2409],[136,-0.2413],[137,-0.2392],[138,-0.2336],[139,-0.2347],[140,-0.2369],[141,-0.236],[142,-0.2252],[143,-0.2237],[145,-0.2378],[146,-0.2385],[147,-0.2385],[148,-0.2376],[149,-0.2299],[150,-0.2289],[151,-0.2226],[152,-0.2188],[153,-0.2274],[154,-0.2282],[155,-0.2341],[156,-0.2419],[157,-0.2399],[158,-0.2401],[159,-0.2414],[160,-0.2386],[161,-0.2419],[162,-0.243],[163,-0.2447],[164,-0.2455],[165,-0.2437],[166,-0.2439],[167,-0.2483],[168,-0.2409],[169,-0.2402],[170,-0.2536],[171,-0.2384],[172,-0.2377],[177,-0.2276],[181,-0.201],[182,-0.2321],[183,-0.2825],[184,0.0103]]
       }
       ball_r_fk_ctrl.Rotation.Z {
        type=MovieSceneScriptingFloatChannel
        default=2.1589
        thinned_from=178
        keys [[0,-0.3317],[1,-0.3238],[2,-0.3792],[3,-0.3157],[4,-0.3103],[5,-0.3203],[6,-0.3113],[7,-0.2851],[8,-0.3056],[9,-0.3316],[10,-0.281],[11,-0.2985],[12,-0.2506],[13,-0.3238],[14,-0.2825],[15,-0.2566],[16,-0.2685],[17,-0.2895],[18,-0.2809],[19,-0.2739],[20,-0.3075],[21,-0.3213],[22,-0.2624],[23,-0.2116],[24,-0.2431],[25,-0.1682],[26,-0.1589],[27,-0.1889],[28,-0.1671],[29,-0.2042],[30,-0.1442],[31,-0.0977],[32,-0.1203],[33,-0.0698],[34,-0.0781],[35,-0.1134],[36,-0.1231],[37,-0.0988],[38,-0.1476],[39,-0.092],[40,-0.0603],[41,-0.0495],[42,-0.0473],[43,-0.0337],[44,-0.0511],[45,-0.0164],[46,0.005],[47,0.003],[48,-0.0235],[49,-0.011],[50,0.0784],[51,-0.0513],[52,-0.0703],[53,-0.1561],[54,-0.1971],[55,-0.3005],[56,-0.1708],[57,-0.1993],[58,-0.2039],[59,-0.1922],[60,-0.1818],[61,-0.224],[62,-0.2333],[63,-0.2224],[64,-0.2793],[65,-0.225],[66,-0.3466],[67,-0.197],[68,0.2499],[69,0.4491],[70,0.2359],[71,-0.08],[72,-0.3185],[73,-0.3914],[74,-0.4556],[75,-0.4239],[76,-0.347],[77,-0.3687],[78,-0.4116],[79,-0.3894],[80,-0.2877],[81,-0.3078],[82,-0.3122],[83,-0.1872],[84,-0.2052],[85,-0.2609],[86,-0.239],[87,-0.2709],[88,-0.2888],[89,-0.2553],[90,-0.1996],[91,-0.1781],[92,-0.0969],[93,-0.1067],[94,-0.0735],[95,-0.0781],[96,-0.1322],[97,-0.1325],[98,-0.0576],[99,-0.111],[100,-0.0623],[101,-0.1338],[102,-0.2475],[103,-0.2743],[104,-0.1229],[105,-0.1082],[106,-0.1451],[107,-0.2182],[108,-0.1576],[109,-0.193],[110,-0.1667],[111,-0.2482],[112,-0.2025],[113,-0.1938],[114,-0.2502],[115,-0.2908],[116,-0.2969],[117,-0.2955],[118,-0.2754],[119,-0.28],[120,-0.2852],[121,-0.2737],[122,-0.2635],[123,-0.236],[124,-0.219],[125,-0.219],[126,-0.284],[127,-0.324],[128,-0.3685],[129,-0.3518],[130,-0.356],[131,-0.375],[132,-0.293],[133,-0.2906],[134,-0.37],[135,-0.3877],[136,-0.392],[137,-0.3704],[138,-0.3148],[139,-0.3254],[140,-0.3478],[141,-0.3387],[142,-0.2292],[143,-0.2144],[145,-0.357],[146,-0.3635],[147,-0.3641],[148,-0.3545],[149,-0.2767],[150,-0.2666],[151,-0.2036],[152,-0.1647],[153,-0.2519],[154,-0.2602],[155,-0.3194],[156,-0.3976],[157,-0.3776],[158,-0.3804],[159,-0.3935],[160,-0.3647],[161,-0.398],[162,-0.4092],[163,-0.4264],[164,-0.434],[165,-0.4161],[166,-0.4183],[167,-0.4622],[168,-0.388],[169,-0.3811],[170,-0.5162],[171,-0.3624],[172,-0.3555],[177,-0.254],[181,0.0138],[182,-0.3027],[183,-0.8091],[184,2.1589]]
       }
       ball_r_fk_ctrl.Scale.X {type=MovieSceneScriptingFloatChannel, default=1, const [1,178]}
       ball_r_fk_ctrl.Scale.Y {type=MovieSceneScriptingFloatChannel, default=1, const [1,178]}
       ball_r_fk_ctrl.Scale.Z {type=MovieSceneScriptingFloatChannel, default=1, const [1,178]}
       foot_r_ik_ctrl.Location.X {
        type=MovieSceneScriptingFloatChannel
        default=-19.7975
        thinned_from=185
        keys [[0,-45.4467],[3,-45.4228],[4,-45.731],[7,-47.3498],[9,-47.8843],[11,-49.3025],[12,-49.9279],[13,-50.0346],[14,-49.7323],[17,-48.3642],[18,-47.8233],[19,-47.5086],[20,-47.3114],[21,-46.7232],[22,-46.4552],[25,-46.0325],[26,-46.0817],[28,-47.3381],[29,-47.4597],[30,-47.6854],[32,-49.0626],[33,-49.4997],[34,-49.4991],[35,-49.1984],[37,-48.9332],[39,-48.3084],[40,-48.17],[41,-47.8611],[42,-47.3327],[43,-47.078],[45,-47.2123],[46,-47.5478],[48,-47.9884],[50,-48.6897],[52,-48.9877],[53,-48.6026],[54,-47.8037],[56,-46.5319],[57,-45.7667],[58,-45.1347],[60,-43.4713],[61,-42.7081],[62,-42.2161],[63,-42.0592],[64,-42.1213],[65,-41.8361],[66,-41.8179],[67,-41.2617],[68,-40.1539],[69,-39.1338],[72,-36.3045],[74,-34.8383],[76,-33.6779],[78,-32.232],[79,-31.6114],[80,-31.0991],[81,-30.7196],[82,-30.4692],[85,-29.9431],[90,-28.842],[91,-28.6586],[97,-27.9411],[100,-27.7448],[107,-27.951],[112,-27.7774],[115,-27.4911],[122,-27.4897],[126,-27.5605],[137,-27.6116],[142,-27.7963],[146,-27.76],[151,-27.8971],[155,-27.7624],[157,-27.5663],[160,-27.1147],[161,-27.033],[163,-26.5231],[165,-25.8525],[166,-25.4331],[167,-24.6561],[168,-22.3805],[169,-19.8186],[170,-17.8644],[171,-17.3683],[172,-17.6165],[173,-18.546],[174,-19.8605],[175,-21.2674],[176,-22.8064],[177,-24.5006],[178,-26.5041],[179,-28.1508],[180,-30.7934],[181,-32.6189],[182,-23.8285],[183,-19.7217],[184,-19.7975]]
       }
       foot_r_ik_ctrl.Location.Y {
        type=MovieSceneScriptingFloatChannel
        default=24.9217
        thinned_from=185
        keys [[0,-100.1118],[13,-99.0494],[30,-99.8584],[46,-99.5925],[49,-98.8758],[51,-98.041],[52,-97.0971],[54,-93.4857],[55,-90.3312],[56,-86.9295],[57,-82.7272],[59,-73.5564],[60,-69.2086],[61,-65.2568],[63,-59.1083],[65,-52.6229],[66,-49.1099],[67,-45.7566],[68,-42.5572],[69,-39.2431],[72,-28.9854],[73,-25.6835],[74,-22.6791],[75,-19.8374],[76,-17.2518],[77,-15.0169],[78,-13.0326],[79,-11.3223],[80,-9.9367],[81,-8.8508],[82,-8.1081],[84,-6.9573],[85,-6.3969],[86,-5.8766],[87,-5.381],[88,-4.8999],[89,-4.4657],[90,-4.063],[91,-3.6706],[93,-2.9035],[94,-2.5554],[95,-2.2313],[96,-1.9478],[97,-1.6975],[98,-1.4847],[99,-1.3776],[100,-1.3545],[101,-1.3103],[102,-1.3015],[103,-1.3241],[104,-1.3407],[106,-1.3572],[107,-1.3606],[108,-1.344],[109,-1.2997],[110,-1.2431],[111,-1.2211],[112,-1.2164],[113,-1.1936],[114,-1.1513],[115,-1.1266],[116,-1.144],[117,-1.1385],[118,-1.1089],[119,-1.0987],[120,-1.1144],[121,-1.1239],[122,-1.1288],[123,-1.1035],[124,-1.0715],[125,-1.0423],[126,-0.975],[127,-0.9207],[128,-0.8634],[129,-0.8536],[130,-0.8657],[131,-0.8992],[132,-0.9471],[133,-0.9646],[134,-0.9861],[135,-1.0331],[136,-1.1523],[137,-1.2593],[138,-1.3269],[139,-1.361],[140,-1.3812],[141,-1.4125],[142,-1.4503],[143,-1.4594],[144,-1.4715],[145,-1.5038],[146,-1.5556],[147,-1.6241],[148,-1.6524],[149,-1.6982],[150,-1.7212],[151,-1.7052],[152,-1.6269],[153,-1.4698],[154,-1.2336],[155,-1.081],[156,-0.9437],[157,-0.8579],[158,-0.7869],[159,-0.733],[160,-0.6956],[161,-0.6848],[162,-0.7047],[163,-0.7656],[164,-0.934],[165,-1.1732],[166,-1.5],[167,-2.0025],[168,-2.9039],[169,-3.2743],[170,-3.7877],[171,-4.3416],[172,-4.5553],[173,-4.5138],[174,-3.9637],[175,-3.1617],[176,-2.7235],[177,-2.5726],[178,-2.7629],[179,-3.3002],[180,-4.2129],[181,-6.1344],[182,-4.1753],[183,19.582],[184,24.9217]]
       }
       foot_r_ik_ctrl.Location.Z {
        type=MovieSceneScriptingFloatChannel
        default=87.4407
        thinned_from=185
        keys [[0,89.4416],[7,89.7649],[10,90.2745],[27,90.5243],[30,90.98],[34,91.1655],[36,91.4494],[45,91.0498],[50,89.4258],[51,88.4475],[52,86.5932],[53,85.0089],[54,82.0373],[55,79.9975],[56,76.4379],[57,73.1804],[58,69.7751],[61,61.0181],[62,57.3176],[63,54.0442],[65,46.9463],[66,43.5109],[67,40.3959],[68,37.6869],[69,35.2879],[70,33.1632],[71,31.3901],[72,29.9409],[73,28.7871],[74,27.9504],[75,27.314],[76,26.8298],[77,26.4913],[78,26.3121],[79,26.2651],[80,26.3337],[81,26.54],[83,27.0968],[88,28.2253],[90,28.5761],[98,29.4789],[100,29.5911],[111,29.5027],[112,29.4621],[114,29.1954],[115,29.0121],[135,28.9208],[137,28.983],[141,28.799],[151,28.7735],[154,28.6989],[157,28.7398],[159,28.5551],[160,28.3789],[161,28.277],[164,27.5516],[165,27.3718],[166,27.1246],[167,26.5807],[168,24.522],[169,22.0737],[170,19.9686],[171,18.0207],[172,16.674],[173,14.9803],[174,13.4561],[175,12.1835],[176,11.4275],[177,11.1683],[178,11.2226],[179,11.5008],[180,12.1084],[181,12.9456],[182,15.4623],[183,49.4332],[184,87.4407]]
       }
       foot_r_ik_ctrl.Rotation.X {
        type=MovieSceneScriptingFloatChannel
        default=79.3073
        thinned_from=185
        keys [[0,130.3763],[7,130.3524],[9,130.7344],[12,130.6442],[34,130.9144],[36,131.2401],[50,130.7145],[52,130.1435],[53,130.2091],[54,129.5376],[55,129.1434],[58,125.9177],[59,124.6579],[60,122.7907],[61,120.163],[63,112.0007],[64,107.6307],[65,103.0354],[66,98.2211],[68,89.1961],[69,85.1051],[70,81.2354],[71,77.6667],[72,74.4282],[74,68.3198],[75,65.4075],[76,62.6972],[77,60.2205],[78,58.1387],[79,56.3613],[80,54.8756],[81,53.9809],[85,52.2203],[88,51.2446],[94,48.8921],[95,48.5656],[98,47.9286],[101,47.8697],[107,48.0411],[111,47.9796],[112,47.8829],[115,47.1797],[121,46.9899],[127,47.0174],[137,47.2968],[143,47.073],[147,47.3399],[154,47.2517],[157,47.4954],[159,47.467],[160,47.3594],[161,47.36],[163,47.081],[165,46.6071],[167,46.2824],[168,43.1523],[169,39.8828],[170,39.4043],[171,39.6545],[172,40.8705],[173,42.5994],[174,44.0045],[176,44.4927],[177,45.5181],[178,46.9833],[179,48.5534],[180,50.6655],[181,51.9829],[182,42.1769],[183,65.02],[184,79.3073]]
       }
       foot_r_ik_ctrl.Rotation.Y {
        type=MovieSceneScriptingFloatChannel
        default=2.9226
        thinned_from=185
        keys [[0,3.746],[1,3.5567],[2,3.4231],[3,3.5002],[4,3.6093],[5,3.8412],[6,4.0822],[7,4.3998],[8,4.7616],[9,4.9478],[10,5.3112],[11,5.596],[12,5.8423],[13,6.1036],[14,6.186],[15,6.1479],[16,6.062],[17,5.8677],[18,5.7211],[20,5.5282],[21,5.199],[22,5.0999],[23,4.9443],[24,4.7377],[25,4.6285],[26,4.4852],[28,4.5704],[30,4.5079],[31,4.6846],[32,4.9774],[33,5.4378],[34,5.5479],[35,5.6962],[36,6.03],[37,6.0817],[38,6.039],[39,6.1366],[40,6.3456],[41,6.2856],[42,6.1567],[43,6.2115],[45,6.2888],[46,6.533],[47,6.4392],[48,6.1278],[49,5.5743],[50,4.8324],[51,4.5832],[52,4.642],[53,4.485],[54,4.601],[55,4.3882],[56,5.6891],[57,6.7938],[58,7.9543],[59,9.3519],[60,11.1274],[61,12.6099],[62,13.7928],[63,14.2695],[64,14.0593],[65,14.1883],[66,13.3511],[67,13.736],[68,14.2179],[69,14.0422],[70,13.9792],[71,14.3911],[72,14.2636],[73,14.0458],[74,13.7275],[75,13.1636],[76,12.4711],[77,11.7019],[78,10.9019],[80,9.354],[81,9.1394],[82,8.9617],[83,8.7249],[84,8.4039],[86,7.8475],[89,7.3803],[90,7.1538],[91,6.8743],[93,6.5126],[94,6.3497],[95,6.215],[96,6.0523],[97,5.8747],[98,5.7119],[99,5.7044],[100,5.4799],[101,5.4763],[102,5.5977],[103,5.6087],[104,5.5548],[105,5.5374],[107,5.6125],[108,5.5197],[109,5.4809],[110,5.5019],[111,5.4861],[112,5.2877],[113,4.9836],[114,4.7002],[115,4.2369],[116,4.0658],[117,4.0299],[118,4.0324],[119,3.9167],[120,3.8622],[121,3.8778],[122,3.8761],[123,3.843],[124,3.8319],[125,3.8065],[126,3.8373],[127,3.848],[128,3.831],[129,3.7476],[130,3.6987],[131,3.6928],[132,3.5713],[133,3.5346],[134,3.4552],[135,3.5421],[136,3.5931],[137,3.5832],[138,3.4812],[139,3.4505],[140,3.3691],[141,3.3002],[142,3.257],[143,3.2489],[144,3.2317],[145,3.2423],[146,3.3053],[147,3.3601],[148,3.3522],[149,3.3658],[150,3.3368],[151,3.261],[152,3.1528],[153,3.1293],[154,3.0078],[155,2.9906],[156,2.9822],[158,2.8209],[159,2.8493],[160,2.7693],[161,2.7324],[162,2.6217],[163,2.6881],[164,2.8636],[165,3.1258],[166,3.604],[167,4.1044],[168,5.883],[169,6.0899],[170,4.6551],[171,5.0809],[172,4.2955],[173,2.2564],[174,2.0583],[175,2.1366],[176,3.0759],[177,3.4546],[178,3.2278],[179,3.9874],[180,5.6785],[181,7.5051],[182,15.107],[183,6.7995],[184,2.9226]]
       }
       foot_r_ik_ctrl.Rotation.Z {
        type=MovieSceneScriptingFloatChannel
        default=-34.8727
        thinned_from=185
        keys [[0,-7.5105],[1,-7.6716],[2,-7.6326],[3,-7.5467],[4,-7.6588],[5,-8.0888],[6,-8.6143],[7,-8.9345],[8,-8.9585],[9,-8.8629],[10,-9.4462],[11,-9.9522],[12,-10.4203],[13,-10.4887],[14,-10.3391],[15,-9.9715],[16,-9.6269],[17,-9.4161],[18,-9.0595],[19,-8.9066],[20,-8.8234],[21,-8.5577],[22,-8.5132],[23,-8.5365],[25,-8.2544],[26,-8.3126],[27,-8.7928],[28,-9.367],[29,-9.3529],[30,-9.5996],[31,-10.1353],[32,-10.7577],[33,-11.3116],[34,-11.2897],[35,-10.985],[37,-10.994],[38,-10.7719],[39,-10.6253],[40,-10.5818],[41,-10.4905],[42,-10.3209],[43,-10.1756],[44,-10.2943],[45,-10.3636],[46,-10.5215],[47,-10.6463],[48,-10.8239],[49,-10.6936],[50,-11.1805],[52,-11.4503],[53,-11.4088],[54,-11.2076],[55,-11.2745],[56,-10.672],[57,-9.5958],[58,-9.2948],[59,-9.6639],[60,-10.1524],[61,-10.9773],[63,-13.8937],[64,-15.5341],[65,-17.077],[66,-17.9774],[67,-19.523],[68,-21.4432],[69,-22.787],[70,-23.5128],[71,-24.7603],[72,-26.4646],[73,-28.3221],[74,-30.6837],[75,-32.6947],[76,-34.6271],[77,-36.1945],[78,-37.1386],[79,-37.8457],[80,-38.3085],[81,-38.8973],[82,-39.292],[85,-40.017],[86,-39.9481],[87,-40.0577],[89,-40.1619],[90,-40.3408],[91,-40.4235],[94,-40.9879],[96,-40.9921],[98,-40.7796],[99,-40.5907],[100,-40.1646],[102,-39.5811],[103,-39.4676],[109,-39.4807],[110,-39.4216],[111,-39.5755],[112,-40.0486],[114,-40.6994],[116,-41.5451],[117,-41.877],[118,-41.9936],[122,-42.0846],[124,-41.9026],[129,-41.7327],[131,-41.5536],[132,-41.5554],[133,-41.4611],[135,-40.9317],[136,-40.829],[139,-41.1064],[142,-41.4991],[143,-41.3765],[145,-40.8247],[147,-40.6474],[150,-40.5943],[152,-40.7301],[153,-40.6887],[154,-40.8635],[155,-40.9183],[156,-41.1075],[159,-42.1438],[160,-42.5735],[161,-42.7541],[162,-43.17],[163,-43.4645],[164,-43.5359],[165,-43.422],[166,-43.1246],[167,-42.5798],[168,-39.889],[169,-36.3884],[170,-33.0243],[171,-33.4922],[172,-35.8653],[173,-39.0709],[174,-42.9062],[175,-46.0173],[176,-48.0689],[177,-48.9117],[178,-50.0125],[179,-50.4034],[180,-51.7858],[181,-52.7051],[182,-48.9501],[183,-31.9612],[184,-34.8727]]
       }
       foot_r_ik_ctrl.Scale.X {type=MovieSceneScriptingFloatChannel, default=1, const [1,185]}
       foot_r_ik_ctrl.Scale.Y {type=MovieSceneScriptingFloatChannel, default=1, const [1,185]}
       foot_r_ik_ctrl.Scale.Z {type=MovieSceneScriptingFloatChannel, default=1, const [1,185]}
       ball_r_ik_ctrl.Location.X {
        type=MovieSceneScriptingFloatChannel
        default=0
        thinned_from=185
        keys [[0,0],[167,0],[168,-0.0002],[178,-0.0001],[179,0],[180,-0.0002],[181,-0.0001],[182,0.0019],[183,-0.0033],[184,0]]
       }
       ball_r_ik_ctrl.Location.Y {
        type=MovieSceneScriptingFloatChannel
        default=-0.0001
        thinned_from=185
        keys [[0,0],[56,-0.0001],[57,-0.0003],[58,-0.0001],[59,-0.0003],[60,-0.0001],[167,0],[168,-0.0002],[170,-0.0003],[172,0],[174,0],[175,-0.0002],[176,0],[178,-0.0001],[179,0],[180,-0.0002],[181,-0.0007],[182,-0.0068],[183,-0.0061],[184,-0.0001]]
       }
       ball_r_ik_ctrl.Location.Z {
        type=MovieSceneScriptingFloatChannel
        default=0
        thinned_from=185
        keys [[0,0],[56,0.0001],[57,0.0002],[167,0],[169,0.0001],[170,-0.0001],[171,0.0001],[180,0],[181,-0.0002],[182,0.0017],[183,0.0052],[184,0]]
       }
       ball_r_ik_ctrl.Rotation.X {
        type=MovieSceneScriptingFloatChannel
        default=2.15
        thinned_from=185
        keys [[0,-0.3528],[1,-0.3449],[2,-0.4006],[3,-0.3367],[4,-0.3314],[5,-0.3414],[6,-0.3324],[7,-0.3061],[8,-0.3266],[9,-0.3528],[10,-0.3019],[11,-0.3195],[12,-0.2713],[13,-0.345],[14,-0.3034],[15,-0.2774],[16,-0.2894],[17,-0.3105],[18,-0.3019],[19,-0.2948],[20,-0.3286],[21,-0.3424],[22,-0.2832],[23,-0.2322],[24,-0.2638],[25,-0.1886],[26,-0.1792],[27,-0.2094],[28,-0.1875],[29,-0.2248],[30,-0.1645],[31,-0.1177],[32,-0.1405],[33,-0.0897],[34,-0.098],[35,-0.1335],[36,-0.1433],[37,-0.1188],[38,-0.1679],[39,-0.112],[40,-0.0802],[41,-0.0693],[42,-0.067],[43,-0.0534],[44,-0.0709],[45,-0.036],[46,-0.0146],[47,-0.0165],[48,-0.0431],[49,-0.0306],[50,0.0593],[51,-0.0711],[52,-0.0902],[53,-0.1764],[54,-0.2176],[55,-0.3215],[56,-0.1912],[57,-0.2198],[58,-0.2244],[59,-0.2127],[60,-0.2022],[61,-0.2446],[62,-0.254],[63,-0.243],[64,-0.3002],[65,-0.2456],[66,-0.3678],[67,-0.2175],[68,0.2316],[69,0.4318],[70,0.2175],[71,-0.0999],[72,-0.3396],[73,-0.4128],[74,-0.4774],[75,-0.4455],[76,-0.3683],[77,-0.39],[78,-0.4331],[79,-0.4108],[80,-0.3086],[81,-0.3288],[82,-0.3333],[83,-0.2077],[84,-0.2258],[85,-0.2817],[86,-0.2597],[87,-0.2918],[88,-0.3098],[89,-0.2761],[90,-0.2201],[91,-0.1985],[92,-0.1169],[93,-0.1268],[94,-0.0934],[95,-0.098],[96,-0.1524],[97,-0.1527],[98,-0.0775],[99,-0.1311],[100,-0.0821],[101,-0.154],[102,-0.2683],[103,-0.2952],[104,-0.143],[105,-0.1283],[106,-0.1653],[107,-0.2388],[108,-0.1779],[109,-0.2135],[110,-0.1871],[111,-0.269],[112,-0.2231],[113,-0.2143],[114,-0.271],[115,-0.3118],[116,-0.3179],[117,-0.3165],[118,-0.2963],[119,-0.3009],[120,-0.3062],[121,-0.2946],[122,-0.2843],[123,-0.2567],[124,-0.2396],[125,-0.2396],[126,-0.305],[127,-0.3451],[128,-0.3899],[129,-0.3731],[130,-0.3773],[131,-0.3963],[132,-0.3139],[133,-0.3116],[134,-0.3914],[135,-0.4091],[136,-0.4134],[137,-0.3918],[138,-0.3359],[139,-0.3465],[140,-0.3691],[141,-0.3599],[142,-0.2499],[143,-0.235],[145,-0.3783],[146,-0.3848],[147,-0.3854],[148,-0.3757],[149,-0.2976],[150,-0.2875],[151,-0.2241],[152,-0.185],[153,-0.2727],[154,-0.281],[155,-0.3405],[156,-0.4191],[157,-0.399],[158,-0.4018],[159,-0.415],[160,-0.3861],[161,-0.4195],[162,-0.4308],[163,-0.448],[164,-0.4557],[165,-0.4377],[166,-0.4399],[167,-0.4841],[168,-0.4095],[169,-0.4026],[170,-0.5383],[171,-0.3838],[172,-0.3768],[173,-0.237],[174,-0.2352],[175,-0.2186],[176,-0.2473],[177,-0.2748],[178,-0.126],[179,-0.1479],[180,-0.1018],[181,-0.0057],[182,-0.3237],[183,-0.8327],[184,2.15]]
       }
       ball_r_ik_ctrl.Rotation.Y {
        type=MovieSceneScriptingFloatChannel
        default=-0.1961
        thinned_from=185
        keys [[0,-0.2027],[50,-0.2018],[55,-0.2027],[65,-0.2025],[66,-0.2027],[67,-0.2024],[68,-0.2014],[69,-0.2008],[72,-0.2027],[79,-0.2028],[101,-0.2023],[103,-0.2026],[141,-0.2027],[152,-0.2023],[167,-0.203],[169,-0.2028],[170,-0.2032],[177,-0.2025],[178,-0.2021],[180,-0.2022],[181,-0.2018],[182,-0.2023],[183,-0.2037],[184,-0.1961]]
       }
       ball_r_ik_ctrl.Rotation.Z {
        type=MovieSceneScriptingFloatChannel
        default=0.0066
        thinned_from=185
        keys [[0,0.055],[1,0.0549],[2,0.0559],[3,0.0547],[4,0.0546],[5,0.0548],[6,0.0546],[7,0.0541],[9,0.055],[10,0.054],[11,0.0544],[12,0.0534],[13,0.0549],[14,0.0541],[15,0.0536],[17,0.0542],[19,0.0539],[20,0.0545],[21,0.0548],[23,0.0527],[24,0.0533],[25,0.0518],[26,0.0517],[27,0.0522],[28,0.0518],[29,0.0525],[30,0.0514],[31,0.0505],[32,0.0509],[33,0.0499],[34,0.0501],[35,0.0508],[36,0.051],[37,0.0505],[38,0.0514],[39,0.0503],[40,0.0497],[42,0.0495],[43,0.0492],[44,0.0495],[45,0.0489],[46,0.0485],[47,0.0485],[48,0.049],[49,0.0488],[50,0.047],[51,0.0496],[52,0.0499],[53,0.0516],[54,0.0524],[55,0.0544],[56,0.0519],[57,0.0524],[58,0.0525],[60,0.0521],[61,0.0529],[62,0.0531],[63,0.0529],[64,0.054],[65,0.0529],[66,0.0553],[67,0.0524],[68,0.0437],[69,0.0398],[70,0.044],[71,0.0501],[72,0.0547],[74,0.0574],[75,0.0568],[76,0.0553],[77,0.0557],[78,0.0566],[79,0.0562],[80,0.0542],[81,0.0546],[82,0.0546],[83,0.0522],[84,0.0526],[85,0.0536],[86,0.0532],[87,0.0538],[88,0.0542],[89,0.0535],[90,0.0524],[91,0.052],[92,0.0504],[93,0.0506],[94,0.05],[95,0.0501],[96,0.0511],[97,0.0511],[98,0.0497],[99,0.0507],[100,0.0498],[101,0.0512],[102,0.0534],[103,0.0539],[104,0.0509],[105,0.0507],[106,0.0514],[107,0.0528],[108,0.0516],[109,0.0523],[110,0.0518],[111,0.0534],[112,0.0525],[113,0.0523],[114,0.0534],[115,0.0542],[117,0.0543],[118,0.0539],[120,0.0541],[122,0.0537],[123,0.0532],[124,0.0528],[125,0.0528],[126,0.0541],[128,0.0557],[129,0.0554],[130,0.0555],[131,0.0559],[132,0.0543],[133,0.0542],[134,0.0558],[135,0.0561],[136,0.0562],[137,0.0558],[138,0.0547],[139,0.0549],[140,0.0553],[141,0.0552],[142,0.053],[143,0.0527],[145,0.0555],[147,0.0557],[148,0.0555],[149,0.0539],[150,0.0537],[151,0.0525],[152,0.0518],[153,0.0534],[154,0.0536],[155,0.0548],[156,0.0563],[157,0.0559],[159,0.0562],[160,0.0556],[161,0.0563],[164,0.057],[165,0.0567],[166,0.0567],[167,0.0576],[168,0.0561],[169,0.056],[170,0.0586],[171,0.0556],[172,0.0555],[173,0.0528],[174,0.0527],[175,0.0524],[177,0.0535],[178,0.0505],[179,0.0511],[180,0.0502],[181,0.0483],[182,0.0544],[183,0.0645],[184,0.0066]]
       }
       ball_r_ik_ctrl.Scale.X {type=MovieSceneScriptingFloatChannel, default=1, const [1,185]}
       ball_r_ik_ctrl.Scale.Y {type=MovieSceneScriptingFloatChannel, default=1, const [1,185]}
       ball_r_ik_ctrl.Scale.Z {type=MovieSceneScriptingFloatChannel, default=1, const [1,185]}
       foot_bk1_r_ctrl.Location.X {type=MovieSceneScriptingFloatChannel, default=0, const [0,185]}
       foot_bk1_r_ctrl.Location.Y {type=MovieSceneScriptingFloatChannel, default=0, const [0,185]}
       foot_bk1_r_ctrl.Location.Z {type=MovieSceneScriptingFloatChannel, default=0, const [0,185]}
       foot_bk1_r_ctrl.Rotation.X {type=MovieSceneScriptingFloatChannel, default=0, const [0,185]}
       foot_bk1_r_ctrl.Rotation.Y {type=MovieSceneScriptingFloatChannel, default=0, const [0,185]}
       foot_bk1_r_ctrl.Rotation.Z {type=MovieSceneScriptingFloatChannel, default=0, const [0,185]}
       foot_bk1_r_ctrl.Scale.X {type=MovieSceneScriptingFloatChannel, default=1, const [1,185]}
       foot_bk1_r_ctrl.Scale.Y {type=MovieSceneScriptingFloatChannel, default=1, const [1,185]}
       foot_bk1_r_ctrl.Scale.Z {type=MovieSceneScriptingFloatChannel, default=1, const [1,185]}
       foot_roll_r_ctrl.Location.X {type=MovieSceneScriptingFloatChannel, default=0, const [0,185]}
       foot_roll_r_ctrl.Location.Y {type=MovieSceneScriptingFloatChannel, default=0, const [0,185]}
       foot_roll_r_ctrl.Location.Z {type=MovieSceneScriptingFloatChannel, default=0, const [0,185]}
       foot_roll_r_ctrl.Rotation.X {type=MovieSceneScriptingFloatChannel, default=0, const [0,185]}
       foot_roll_r_ctrl.Rotation.Y {type=MovieSceneScriptingFloatChannel, default=0, const [0,185]}
       foot_roll_r_ctrl.Rotation.Z {type=MovieSceneScriptingFloatChannel, default=0, const [0,185]}
       foot_roll_r_ctrl.Scale.X {type=MovieSceneScriptingFloatChannel, default=1, const [1,185]}
       foot_roll_r_ctrl.Scale.Y {type=MovieSceneScriptingFloatChannel, default=1, const [1,185]}
       foot_roll_r_ctrl.Scale.Z {type=MovieSceneScriptingFloatChannel, default=1, const [1,185]}
       roll_blend_r_ctrl {type=MovieSceneScriptingFloatChannel, default=1, const [1,185]}
       tip_r_ctrl.Location.X {type=MovieSceneScriptingFloatChannel, default=0, const [0,185]}
       tip_r_ctrl.Location.Y {type=MovieSceneScriptingFloatChannel, default=0, const [0,185]}
       tip_r_ctrl.Location.Z {type=MovieSceneScriptingFloatChannel, default=0, const [0,185]}
       tip_r_ctrl.Rotation.X {type=MovieSceneScriptingFloatChannel, default=0, const [0,185]}
       tip_r_ctrl.Rotation.Y {type=MovieSceneScriptingFloatChannel, default=0, const [0,185]}
       tip_r_ctrl.Rotation.Z {type=MovieSceneScriptingFloatChannel, default=0, const [0,185]}
       tip_r_ctrl.Scale.X {type=MovieSceneScriptingFloatChannel, default=1, const [1,185]}
       tip_r_ctrl.Scale.Y {type=MovieSceneScriptingFloatChannel, default=1, const [1,185]}
       tip_r_ctrl.Scale.Z {type=MovieSceneScriptingFloatChannel, default=1, const [1,185]}
       heel_r_ctrl.Location.X {type=MovieSceneScriptingFloatChannel, default=0, const [0,185]}
       heel_r_ctrl.Location.Y {type=MovieSceneScriptingFloatChannel, default=0, const [0,185]}
       heel_r_ctrl.Location.Z {type=MovieSceneScriptingFloatChannel, default=0, const [0,185]}
       heel_r_ctrl.Rotation.X {type=MovieSceneScriptingFloatChannel, default=0, const [0,185]}
       heel_r_ctrl.Rotation.Y {type=MovieSceneScriptingFloatChannel, default=0, const [0,185]}
       heel_r_ctrl.Rotation.Z {type=MovieSceneScriptingFloatChannel, default=0, const [0,185]}
       heel_r_ctrl.Scale.X {type=MovieSceneScriptingFloatChannel, default=1, const [1,185]}
       heel_r_ctrl.Scale.Y {type=MovieSceneScriptingFloatChannel, default=1, const [1,185]}
       heel_r_ctrl.Scale.Z {type=MovieSceneScriptingFloatChannel, default=1, const [1,185]}
       leg_r_softness {type=MovieSceneScriptingFloatChannel, default=0, const [0,185]}
       leg_r_pv_ik_ctrl.Location.X {
        type=MovieSceneScriptingFloatChannel
        default=-51.1622
        thinned_from=185
        keys [[0,32.1216],[3,31.9377],[5,31.6328],[6,31.5608],[7,31.5891],[8,31.7747],[9,31.8657],[11,31.6053],[12,31.5731],[13,31.6877],[15,32.1881],[18,32.517],[21,32.6718],[24,32.6356],[26,32.4897],[28,32.2676],[29,32.3207],[30,32.2654],[31,32.0824],[32,31.9916],[34,32.2491],[36,32.7802],[41,33.3846],[43,33.6597],[46,33.6203],[47,33.474],[48,33.0351],[49,32.3038],[50,31.3803],[51,30.3913],[52,28.4097],[53,25.408],[54,22.2311],[55,17.9167],[56,13.7502],[57,9.2321],[58,4.9492],[59,1.3355],[60,-1.4066],[61,-3.659],[62,-5.5719],[63,-7.8792],[64,-10.4532],[65,-12.3124],[66,-14.8805],[68,-18.7304],[69,-20.2523],[70,-21.0395],[71,-20.9522],[72,-20.0675],[73,-18.8075],[76,-15.4892],[77,-14.6595],[78,-14.1602],[79,-13.7445],[80,-13.2666],[81,-12.6933],[82,-12.2227],[83,-11.8966],[84,-11.7231],[85,-11.5859],[87,-11.1081],[88,-10.9358],[89,-10.8205],[90,-10.7471],[91,-10.6241],[92,-10.4322],[94,-10.1275],[95,-10.0113],[96,-9.9431],[97,-9.9099],[98,-9.9569],[99,-10.1323],[100,-10.6582],[101,-10.917],[102,-11.1983],[103,-11.4289],[104,-11.6103],[105,-11.6749],[106,-11.82],[107,-11.877],[108,-11.981],[109,-11.9932],[110,-11.8993],[111,-11.737],[112,-11.5442],[114,-11.0377],[115,-10.6504],[116,-10.4334],[117,-10.1643],[118,-9.9419],[119,-9.8582],[121,-9.913],[125,-10.1104],[128,-10.0651],[129,-10.0789],[130,-10.1541],[132,-10.4043],[133,-10.4521],[134,-10.5551],[135,-10.6853],[136,-10.88],[137,-10.9891],[138,-10.9729],[139,-10.8868],[140,-10.7551],[141,-10.6814],[142,-10.7001],[144,-10.9385],[145,-11.1131],[147,-11.4127],[148,-11.4746],[150,-11.6989],[151,-11.7701],[152,-11.7499],[153,-11.6081],[154,-11.2953],[155,-11.0205],[156,-10.7227],[157,-10.4902],[158,-10.2186],[160,-9.7527],[161,-9.6599],[162,-9.5067],[163,-9.3742],[164,-9.3613],[165,-9.4964],[166,-9.8774],[167,-10.6592],[168,-12.9348],[169,-15.1737],[170,-16.7912],[171,-17.1763],[172,-16.5562],[173,8.5228],[174,-12.4083],[175,12.6081],[176,-15.7006],[177,-14.5996],[178,16.5202],[179,-18.317],[180,15.6918],[181,13.7548],[182,10.5765],[183,-57.088],[184,-51.1622]]
       }
       leg_r_pv_ik_ctrl.Location.Y {
        type=MovieSceneScriptingFloatChannel
        default=27.2304
        thinned_from=185
        keys [[0,-88.3087],[9,-88.88],[12,-89.3046],[24,-88.8032],[27,-89.3064],[32,-90.3447],[49,-88.4827],[50,-88.5985],[51,-87.8008],[52,-85.7432],[53,-82.9587],[54,-78.641],[55,-73.9224],[56,-67.5896],[57,-60.785],[58,-53.6338],[59,-46.6287],[60,-39.8346],[61,-33.7911],[62,-28.6778],[63,-24.7117],[64,-20.9608],[65,-17.04],[67,-10.5178],[68,-7.5797],[69,-4.7086],[70,-1.7408],[71,1.244],[72,4.2335],[73,6.8389],[74,8.904],[75,10.5684],[76,11.9346],[77,12.9891],[78,13.7836],[79,14.3942],[80,14.8786],[81,15.2832],[82,15.568],[83,15.7546],[86,16.0697],[88,16.2046],[92,16.1583],[95,16.1398],[99,16.0631],[100,15.9872],[104,15.8866],[105,15.8884],[106,15.9291],[107,15.9196],[110,15.8079],[112,15.8172],[114,15.9172],[115,16.014],[118,16.208],[123,16.2815],[127,16.239],[130,16.2667],[133,16.2106],[135,16.0736],[137,15.9037],[147,15.5838],[151,15.5516],[152,15.5762],[153,15.6416],[160,16.5505],[162,16.6753],[163,16.7064],[164,16.6972],[165,16.6419],[166,16.4881],[167,16.0418],[168,14.6381],[169,12.7029],[170,10.8895],[171,9.794],[172,9.9031],[173,1.6815],[174,17.309],[175,-0.8334],[176,16.9294],[177,17.2412],[178,-2.9173],[179,15.9052],[180,-3.7273],[181,-4.0191],[182,8.1773],[183,25.6983],[184,27.2304]]
       }
       leg_r_pv_ik_ctrl.Location.Z {
        type=MovieSceneScriptingFloatChannel
        default=56.3533
        thinned_from=185
        keys [[0,6.442],[2,6.4372],[4,6.4133],[5,6.4544],[6,6.5436],[7,6.6533],[8,6.7783],[9,6.8713],[10,6.9452],[12,6.9984],[13,7.0058],[14,6.9979],[17,6.9993],[18,6.9895],[20,6.9376],[22,6.8629],[23,6.8065],[24,6.7654],[25,6.751],[26,6.7752],[27,6.8629],[28,7.0342],[29,7.161],[30,7.2448],[31,7.3012],[32,7.3816],[33,7.4814],[34,7.5036],[36,7.5872],[37,7.5711],[39,7.4365],[40,7.3845],[41,7.3581],[42,7.3509],[43,7.2898],[44,7.1729],[45,7.0845],[46,6.829],[47,6.5461],[48,6.1835],[49,5.8551],[50,5.7261],[51,4.9269],[52,3.2616],[53,1.6412],[54,-0.7338],[55,-2.2694],[56,-4.2259],[57,-5.4446],[58,-5.9424],[59,-5.681],[60,-4.9092],[61,-3.7541],[62,-2.3116],[63,-0.922],[64,0.6889],[65,2.7626],[66,4.9005],[67,7.2737],[68,9.7689],[69,12.5846],[70,15.6834],[71,18.9906],[72,22.5404],[73,25.9587],[74,29.1714],[75,32.1767],[76,34.9184],[77,37.3968],[78,39.6598],[79,41.6572],[80,43.3321],[81,44.7126],[82,45.745],[88,50.665],[90,52.1155],[94,54.4018],[96,55.3469],[98,56.0392],[100,56.3609],[107,56.0508],[112,56.4201],[124,56.0582],[128,56.2503],[137,56.116],[152,55.6516],[157,56.0676],[167,55.4101],[168,55.0462],[169,54.7976],[170,54.0731],[172,51.094],[173,37.4313],[174,44.7548],[175,34.3855],[176,46.0521],[177,44.5943],[178,29.5615],[179,47.6008],[180,27.2939],[181,26.7309],[182,28.6988],[183,43.1038],[184,56.3533]]
       }
       leg_r_pv_ik_ctrl.Rotation.X {
        type=MovieSceneScriptingFloatChannel
        default=41.277
        thinned_from=185
        keys [[0,78.6261],[27,79.7191],[37,80.3359],[45,80.3345],[49,79.6038],[50,79.5699],[51,78.4432],[52,75.8005],[53,72.6371],[55,62.926],[56,56.4848],[57,49.8966],[58,43.1774],[59,37.0802],[60,31.4075],[61,26.4861],[62,21.7064],[63,18.0057],[64,14.0527],[65,9.743],[66,5.6738],[67,1.57],[68,-2.4389],[69,-6.5518],[70,-10.7059],[72,-18.8146],[73,-22.5659],[74,-26.02],[75,-29.1802],[76,-32.0068],[77,-34.5397],[78,-36.8313],[79,-38.846],[80,-40.5237],[81,-41.9021],[82,-42.9518],[87,-47.293],[90,-49.5865],[93,-51.5056],[96,-53.062],[98,-53.7318],[100,-54.0003],[105,-53.8728],[106,-53.6749],[112,-53.9587],[124,-53.4327],[128,-53.6669],[137,-53.4589],[141,-53.2327],[146,-53.235],[151,-52.9848],[153,-53.1297],[156,-53.4873],[159,-53.3532],[166,-52.4648],[167,-52.4509],[168,-52.6024],[169,-54.1085],[170,-56.1141],[171,-57.6447],[172,-58.2186],[173,-56.3689],[174,-51.4269],[175,-57.9367],[176,-51.944],[177,-51.3878],[178,-56.1882],[179,-52.1568],[180,-53.6549],[181,-51.8444],[182,-42.7498],[183,9.6351],[184,41.277]]
       }
       leg_r_pv_ik_ctrl.Rotation.Y {
        type=MovieSceneScriptingFloatChannel
        default=0.3803
        thinned_from=185
        keys [[0,-5.6078],[1,-5.6811],[2,-5.7275],[3,-5.747],[4,-5.7086],[5,-5.6507],[6,-5.4516],[7,-5.168],[8,-4.7788],[9,-4.5297],[10,-4.3554],[11,-4.1557],[12,-3.8862],[13,-3.7018],[14,-3.5529],[15,-3.5074],[16,-3.6206],[17,-3.7916],[19,-4.0784],[20,-4.1472],[21,-4.3765],[22,-4.5275],[24,-4.7688],[25,-4.9673],[26,-5.1024],[27,-5.0977],[28,-5.0453],[29,-5.0221],[30,-5.051],[31,-5.0052],[32,-4.8335],[33,-4.4966],[34,-4.3493],[35,-4.1541],[36,-3.9372],[37,-3.7701],[38,-3.7632],[39,-3.7669],[40,-3.6853],[41,-3.7201],[42,-3.8281],[43,-3.8408],[45,-3.7194],[46,-3.4838],[47,-3.4166],[48,-3.6363],[49,-4.1346],[50,-4.9015],[51,-5.2247],[52,-5.5803],[53,-6.5707],[54,-6.852],[55,-7.7025],[56,-7.5369],[57,-7.5038],[58,-7.2538],[59,-6.7491],[60,-5.8301],[61,-4.9901],[62,-4.4892],[63,-4.8179],[64,-5.6152],[65,-5.9666],[66,-7.2549],[67,-8.0147],[68,-8.4776],[69,-8.7817],[70,-8.8673],[71,-8.614],[72,-8.1806],[73,-8.0073],[74,-8.0396],[75,-8.4132],[76,-8.8498],[77,-9.1907],[79,-9.9662],[80,-10.305],[81,-10.6099],[83,-11.1022],[85,-11.6689],[86,-11.8797],[87,-12.0416],[89,-12.4319],[91,-12.6821],[94,-13.1189],[95,-13.231],[98,-13.4924],[99,-13.5152],[100,-13.4357],[102,-13.4317],[104,-13.3639],[107,-13.3208],[110,-13.1804],[111,-13.0921],[112,-12.9593],[114,-12.6334],[115,-12.4027],[116,-12.3181],[118,-12.2728],[120,-12.1569],[122,-12.1369],[124,-12.2204],[130,-12.2465],[132,-12.1789],[134,-12.1492],[135,-12.1051],[136,-12.0114],[137,-11.946],[138,-11.9185],[143,-11.9244],[145,-11.8619],[147,-11.7738],[150,-11.7952],[151,-11.8194],[152,-11.8774],[153,-11.9818],[154,-12.1279],[155,-12.2202],[156,-12.2664],[157,-12.2233],[159,-12.0117],[160,-11.8673],[161,-11.7933],[162,-11.5569],[163,-11.2605],[165,-10.3894],[166,-9.8994],[167,-9.1837],[168,-7.5527],[169,-5.9473],[170,-4.7597],[171,-4.7917],[172,-5.4185],[173,-4.702],[174,-6.2922],[175,-9.2057],[176,-6.9143],[177,-8.4784],[178,-14.679],[179,-5.7917],[180,-15.5621],[181,-13.5652],[182,-1.283],[183,-0.8677],[184,0.3803]]
       }
       leg_r_pv_ik_ctrl.Rotation.Z {
        type=MovieSceneScriptingFloatChannel
        default=32.9975
        thinned_from=185
        keys [[0,33.6885],[3,33.8098],[4,33.6362],[7,32.655],[9,32.3598],[12,31.0734],[13,31.0056],[14,31.2017],[17,32.0546],[18,32.3818],[20,32.6753],[21,33.0457],[22,33.2151],[25,33.5157],[26,33.4982],[28,32.6906],[29,32.6315],[30,32.4987],[32,31.613],[33,31.3195],[34,31.3175],[35,31.5016],[37,31.6157],[39,31.9277],[40,31.95],[41,32.0937],[42,32.382],[43,32.4768],[45,32.3222],[46,32.0444],[48,31.5887],[49,31.2733],[50,30.8751],[51,30.6926],[52,30.3961],[53,30.3276],[54,30.6461],[56,30.6493],[57,30.5394],[58,30.1758],[59,29.7469],[62,27.8319],[65,26.2533],[66,26.0658],[67,26.054],[68,26.593],[69,26.8672],[70,26.5556],[71,25.667],[72,24.127],[73,22.298],[74,20.7846],[75,19.4239],[76,18.1861],[77,17.3767],[78,16.9606],[80,16.3564],[81,15.9493],[82,15.6182],[83,15.4263],[84,15.4223],[85,15.461],[87,15.3591],[88,15.3844],[90,15.5841],[91,15.634],[95,15.5648],[97,15.6735],[98,15.7971],[99,16.007],[100,16.4982],[102,16.9357],[103,17.0852],[104,17.1704],[105,17.1926],[106,17.2682],[107,17.3013],[108,17.3791],[109,17.374],[110,17.2737],[111,17.0765],[112,16.8302],[114,16.1602],[115,15.6391],[116,15.3612],[117,15.0279],[118,14.7597],[119,14.6293],[122,14.6441],[125,14.7706],[127,14.749],[128,14.7113],[129,14.7079],[130,14.7693],[132,14.9884],[133,15.0306],[135,15.2407],[136,15.4069],[137,15.4901],[138,15.4485],[140,15.2264],[141,15.1579],[142,15.1956],[144,15.4485],[145,15.5997],[147,15.8211],[148,15.8676],[151,16.1297],[152,16.1375],[153,16.0417],[155,15.5755],[157,15.0792],[159,14.5084],[160,14.2471],[161,14.1402],[163,13.8091],[164,13.7534],[165,13.8678],[166,14.3098],[167,15.3961],[168,18.8051],[169,22.4973],[170,25.576],[171,27.1754],[172,27.0447],[173,-23.8553],[174,13.1384],[175,-26.947],[176,16.4688],[177,15.2488],[178,-31.1649],[179,18.2266],[180,-33.5291],[181,-34.9708],[182,-31.121],[183,38.3394],[184,32.9975]]
       }
       leg_r_pv_ik_ctrl.Scale.X {type=MovieSceneScriptingFloatChannel, default=1, const [1,185]}
       leg_r_pv_ik_ctrl.Scale.Y {type=MovieSceneScriptingFloatChannel, default=1, const [1,185]}
       leg_r_pv_ik_ctrl.Scale.Z {type=MovieSceneScriptingFloatChannel, default=1, const [1,185]}
       upperarm_l_fk_ctrl.Location.X {
        type=MovieSceneScriptingFloatChannel
        default=0
        thinned_from=175
        keys [[0,0],[165,0],[166,-0.0002],[167,0],[170,-0.0006],[171,-0.0005],[172,-0.0001],[182,0],[183,0.0002],[184,0]]
       }
       upperarm_l_fk_ctrl.Location.Y {
        type=MovieSceneScriptingFloatChannel
        default=-0.0003
        thinned_from=175
        keys [[0,0],[167,0],[170,-0.0003],[172,0],[182,0],[183,-0.0006],[184,-0.0003]]
       }
       upperarm_l_fk_ctrl.Location.Z {type=MovieSceneScriptingFloatChannel, default=-0.0003, thinned_from=175, keys [[0,0],[182,0],[183,-0.0027],[184,-0.0003]]}
       upperarm_l_fk_ctrl.Rotation.X {
        type=MovieSceneScriptingFloatChannel
        default=77.635
        thinned_from=175
        keys [[0,101.5528],[4,101.6624],[5,101.524],[6,101.6079],[8,100.9787],[10,100.6632],[13,100.3324],[14,99.9672],[15,99.9271],[21,101.2533],[24,102.1289],[27,102.1003],[28,101.8781],[29,101.4325],[32,101.0518],[34,101.2316],[35,101.0224],[36,100.6063],[38,101.1203],[39,101.0218],[41,101.2813],[42,101.0472],[44,101.1736],[45,100.9702],[46,100.3896],[47,99.4738],[48,98.2399],[49,96.7085],[50,94.9266],[52,90.7991],[55,83.9036],[56,80.9521],[57,77.07],[59,68.483],[60,64.7628],[61,62.0815],[62,60.9313],[63,61.0584],[64,61.7621],[65,62.8643],[67,65.5513],[71,70.0691],[72,71.0327],[74,72.3246],[75,72.8414],[76,73.1961],[77,73.3796],[79,73.0432],[85,70.4274],[87,70.778],[96,70.7971],[98,72.3042],[99,72.8655],[100,73.6801],[101,74.1487],[102,74.2865],[103,74.0652],[106,72.6511],[107,72.4561],[112,71.2829],[115,69.7042],[118,68.8588],[124,68.2736],[131,67.4207],[136,67.4997],[138,67.6083],[141,67.1893],[161,67.3894],[163,67.7203],[164,67.6667],[165,67.3103],[166,65.2179],[167,58.2139],[168,39.8397],[169,21.8145],[170,6.3175],[171,-9.2075],[172,-17.9956],[174,-29.3452],[176,-28.2534],[177,-26.2253],[179,-22.6693],[181,-6.5653],[182,12.9758],[183,72.1724],[184,77.635]]
       }
       upperarm_l_fk_ctrl.Rotation.Y {
        type=MovieSceneScriptingFloatChannel
        default=31.6141
        thinned_from=175
        keys [[0,37.9885],[9,38.0296],[11,38.162],[15,38.0185],[22,38.4501],[26,38.8183],[28,38.8267],[30,38.6677],[34,38.7043],[35,38.6453],[40,38.7045],[42,38.6525],[44,38.736],[45,38.982],[46,39.5545],[47,40.3797],[50,43.1682],[51,43.9332],[52,44.5113],[53,44.8396],[54,44.8552],[55,44.4953],[56,42.7756],[57,39.1942],[58,34.4119],[59,29.0894],[60,23.8876],[61,19.4672],[62,16.4889],[63,14.6382],[64,13.108],[65,11.8205],[66,10.6979],[68,8.6366],[69,7.5716],[70,6.4401],[71,5.1843],[72,3.9167],[74,1.6262],[75,0.3255],[76,-0.8354],[77,-3.0174],[78,-5.1193],[79,-5.4851],[84,-6.4868],[85,-6.1384],[86,-5.8973],[87,-5.8738],[96,-5.9204],[97,-6.0027],[98,-6.4043],[99,-6.9939],[100,-8.0734],[101,-8.3199],[102,-8.4546],[103,-8.4013],[106,-8.1069],[107,-8.0316],[108,-7.8934],[109,-7.8713],[110,-7.8776],[111,-7.9345],[112,-8.0491],[113,-7.8735],[114,-7.5645],[115,-7.0563],[116,-6.8468],[117,-6.6898],[118,-6.5871],[119,-6.5373],[121,-6.5901],[123,-6.6708],[125,-6.6306],[126,-6.6366],[127,-6.6752],[128,-6.7306],[129,-6.7191],[131,-6.6439],[132,-6.6223],[133,-6.6348],[135,-6.7173],[137,-7.0291],[138,-7.1415],[139,-7.1678],[142,-6.8975],[143,-6.8407],[144,-6.922],[145,-7.0714],[146,-7.1908],[147,-7.2793],[148,-7.3095],[149,-7.2891],[153,-6.9972],[154,-7],[157,-6.9045],[159,-6.7945],[160,-6.6822],[161,-6.6503],[162,-6.5843],[164,-6.5057],[165,-6.3815],[166,-5.2505],[167,-1.4462],[168,8.5986],[169,18.6739],[170,26.6605],[171,30.2914],[172,29.5442],[174,22.918],[176,19.6465],[177,19.9991],[179,23.2204],[181,30.431],[182,62.9076],[183,57.5884],[184,31.6141]]
       }
       upperarm_l_fk_ctrl.Rotation.Z {
        type=MovieSceneScriptingFloatChannel
        default=171.0627
        thinned_from=175
        keys [[0,219.6805],[44,219.4915],[45,218.9692],[46,217.6096],[47,215.5022],[48,212.2523],[49,207.5289],[50,201.5166],[51,194.4003],[52,186.3647],[53,177.5948],[54,168.2752],[55,158.5907],[56,147.1831],[57,133.4113],[59,103.8177],[60,90.5175],[61,79.8958],[62,73.2135],[63,69.2156],[64,65.8481],[65,63.1733],[66,61.2531],[67,60.1496],[68,59.925],[75,62.0103],[76,62.1517],[77,61.9949],[78,61.6233],[79,61.4867],[85,59.3358],[96,58.9549],[98,59.0475],[99,58.8935],[100,58.3896],[103,58.2137],[106,57.2598],[107,57.128],[109,56.5758],[112,56.261],[114,55.7978],[115,55.4685],[118,55.0318],[122,54.9342],[128,54.294],[134,54.2874],[135,54.1127],[136,53.804],[138,53.4643],[145,53.0137],[150,53.2267],[152,53.0387],[154,52.5003],[158,52.035],[159,51.9928],[161,52.1747],[163,52.6564],[165,52.6414],[166,52.0067],[167,49.7656],[168,43.4724],[169,36.3743],[170,28.1195],[171,16.5789],[172,9.0956],[174,-2.4597],[176,-2.9464],[177,-1.0454],[179,4.6649],[181,18.0704],[182,58.9447],[183,156.0459],[184,171.0627]]
       }
       upperarm_l_fk_ctrl.Scale.X {type=MovieSceneScriptingFloatChannel, default=1, const [1,175]}
       upperarm_l_fk_ctrl.Scale.Y {type=MovieSceneScriptingFloatChannel, default=1, const [1,175]}
       upperarm_l_fk_ctrl.Scale.Z {type=MovieSceneScriptingFloatChannel, default=1, const [1,175]}
       lowerarm_l_fk_ctrl.Location.X {
        type=MovieSceneScriptingFloatChannel
        default=-0.0008
        thinned_from=177
        keys [[0,0],[47,-0.0002],[55,-0.0017],[60,-0.0039],[62,-0.0043],[63,-0.0042],[65,-0.0034],[68,-0.0018],[69,-0.001],[70,-0.0005],[72,-0.0016],[74,0],[75,-0.0003],[76,-0.0001],[77,-0.0009],[78,-0.0001],[84,-0.0001],[85,0.0002],[86,0],[96,0.0002],[98,-0.0001],[101,0.0002],[102,-0.0003],[104,0],[105,-0.0002],[107,0.0002],[108,-0.0003],[111,0.0003],[112,-0.0002],[113,-0.0003],[115,0.0004],[117,0.0003],[118,0.0007],[119,-0.0003],[120,0],[121,-0.0005],[122,-0.0001],[123,-0.0005],[124,0.0004],[125,-0.0003],[126,0.0004],[127,-0.0008],[129,-0.0009],[130,-0.0003],[131,-0.0001],[132,0.0003],[133,-0.0011],[134,0.0004],[135,0.0001],[137,-0.0012],[138,0.0005],[139,0.0007],[140,0.0006],[141,0.0003],[142,-0.0012],[143,0.0007],[144,0.0002],[145,-0.0019],[146,-0.0001],[147,0.0008],[149,-0.0018],[150,-0.0011],[151,-0.0012],[152,-0.001],[153,-0.0001],[154,-0.0029],[155,0.0008],[156,0.0009],[157,-0.0025],[158,-0.0014],[159,-0.0028],[160,-0.0028],[162,-0.0001],[163,-0.0005],[164,0.0005],[165,0.0007],[166,-0.0074],[167,-0.002],[168,-0.0041],[169,-0.0052],[170,-0.0072],[171,-0.0042],[172,-0.0009],[174,-0.0013],[176,0],[177,-0.0005],[179,0.0002],[182,0],[183,-0.0262],[184,-0.0008]]
       }
       lowerarm_l_fk_ctrl.Location.Y {
        type=MovieSceneScriptingFloatChannel
        default=0.0004
        thinned_from=177
        keys [[0,0],[55,-0.0008],[61,-0.002],[63,-0.0019],[67,-0.0002],[68,0],[70,-0.0003],[72,0.0007],[74,-0.0001],[75,0.0004],[76,0.0002],[77,0.0006],[78,0.0002],[84,-0.0002],[85,0.0003],[86,0.0001],[87,0.0004],[96,0.0004],[98,-0.0001],[99,0.0002],[101,0.0004],[102,-0.0005],[104,0.0001],[105,-0.0004],[106,-0.0001],[107,0.0004],[108,-0.0006],[110,0.0003],[111,0.0005],[112,-0.0004],[113,-0.0006],[115,0.0008],[116,0.0008],[117,0.0004],[118,0.0012],[119,-0.0005],[120,0.0001],[121,-0.0009],[122,-0.0003],[123,-0.0008],[124,0.0008],[125,-0.0006],[126,0.0008],[127,-0.0014],[128,-0.0017],[129,-0.0017],[130,-0.0005],[131,-0.0003],[132,0.0005],[133,-0.0019],[134,0.0008],[135,0.0002],[136,-0.0011],[137,-0.0023],[138,0.0009],[139,0.0012],[140,0.001],[141,0.0005],[142,-0.0021],[143,0.0013],[144,0.0004],[145,-0.0034],[146,-0.0002],[147,0.0014],[148,-0.0011],[149,-0.0033],[150,-0.0019],[151,-0.0022],[152,-0.0018],[153,-0.0001],[154,-0.0052],[155,0.0015],[156,0.0016],[157,-0.0046],[158,-0.0026],[159,-0.005],[160,-0.0051],[162,-0.0002],[163,-0.0008],[164,0.0011],[165,0.0012],[167,-0.0036],[168,0.0034],[169,0.0021],[170,0.0025],[171,0.005],[172,0.0004],[174,-0.0016],[176,0.0001],[177,-0.0006],[179,0.0003],[182,0],[183,0.0172],[184,0.0004]]
       }
       lowerarm_l_fk_ctrl.Location.Z {
        type=MovieSceneScriptingFloatChannel
        default=0.0001
        thinned_from=177
        keys [[0,0],[68,-0.0009],[70,-0.0001],[72,-0.0007],[74,0],[75,-0.0002],[79,0.0001],[84,0],[85,-0.0001],[98,0],[101,-0.0002],[102,0.0002],[104,0],[105,0.0001],[107,-0.0002],[108,0.0002],[111,-0.0002],[112,0.0002],[113,0.0002],[115,-0.0003],[117,-0.0002],[118,-0.0006],[119,0.0002],[120,-0.0001],[121,0.0004],[122,0.0001],[123,0.0004],[124,-0.0004],[125,0.0003],[126,-0.0004],[127,0.0007],[129,0.0009],[130,0.0003],[131,0.0001],[132,-0.0002],[133,0.001],[134,-0.0004],[135,-0.0001],[137,0.0012],[138,-0.0005],[139,-0.0007],[140,-0.0006],[141,-0.0003],[142,0.0012],[143,-0.0007],[144,-0.0002],[145,0.0019],[146,0.0001],[147,-0.0008],[149,0.0018],[150,0.0011],[151,0.0012],[152,0.001],[153,0.0001],[154,0.003],[155,-0.0009],[156,-0.0009],[157,0.0026],[158,0.0015],[159,0.0029],[160,0.0029],[162,0.0001],[163,0.0004],[164,-0.0006],[165,-0.0007],[166,-0.0034],[167,0.0036],[168,-0.0061],[169,-0.0051],[170,-0.0048],[171,-0.0019],[172,-0.0003],[174,0],[177,-0.0002],[179,0],[182,0],[183,-0.004],[184,0.0001]]
       }
       lowerarm_l_fk_ctrl.Rotation.X {
        type=MovieSceneScriptingFloatChannel
        default=1.1891
        thinned_from=177
        keys [[0,1.0583],[2,1.0468],[6,1.0394],[8,1.0336],[11,1.0331],[12,1.0282],[13,1.0275],[14,1.0329],[17,1.0389],[18,1.0387],[19,1.0418],[20,1.0424],[22,1.0395],[23,1.0398],[27,1.0316],[28,1.0337],[31,1.0323],[32,1.0386],[33,1.0389],[34,1.037],[35,1.0396],[37,1.05],[38,1.0491],[39,1.0449],[40,1.0363],[41,1.0301],[43,1.0351],[44,1.0418],[46,1.0393],[49,1.0278],[55,0.9888],[56,0.9793],[57,0.9656],[58,0.948],[59,0.927],[60,0.9028],[61,0.8758],[62,0.8463],[63,0.8037],[64,0.7408],[65,0.6631],[66,0.5763],[67,0.486],[68,0.3979],[69,0.293],[70,0.1651],[71,0.0224],[72,-0.1265],[74,-0.4086],[75,-0.5544],[76,-0.6756],[77,-0.9017],[78,-1.1564],[79,-1.2328],[84,-1.8471],[85,-1.8596],[86,-1.8603],[87,-1.8548],[89,-1.8331],[96,-1.823],[97,-1.7561],[98,-1.6367],[99,-1.4998],[100,-1.275],[101,-1.2309],[102,-1.2139],[103,-1.233],[104,-1.2423],[105,-1.2553],[106,-1.3461],[107,-1.3684],[108,-1.381],[109,-1.3657],[110,-1.3578],[111,-1.3554],[112,-1.3703],[114,-1.372],[115,-1.3557],[117,-1.3518],[118,-1.356],[119,-1.364],[120,-1.3678],[121,-1.367],[122,-1.3698],[123,-1.3619],[126,-1.3681],[128,-1.3584],[130,-1.3571],[131,-1.3593],[132,-1.3688],[134,-1.3719],[135,-1.3672],[137,-1.3231],[138,-1.313],[140,-1.321],[142,-1.3127],[143,-1.3236],[144,-1.3204],[145,-1.3123],[146,-1.3259],[148,-1.3309],[149,-1.3301],[150,-1.333],[151,-1.333],[153,-1.3419],[154,-1.3294],[155,-1.3334],[156,-1.3341],[157,-1.326],[159,-1.3186],[161,-1.3272],[162,-1.3276],[163,-1.3244],[165,-1.3604],[166,-1.3854],[167,-1.3957],[168,-1.2017],[169,-1.2446],[170,-1.5035],[171,-1.8152],[174,-2.4609],[176,-2.4782],[177,-2.4685],[179,-2.4315],[181,0.9689],[182,1.0123],[183,0.3202],[184,1.1891]]
       }
       lowerarm_l_fk_ctrl.Rotation.Y {
        type=MovieSceneScriptingFloatChannel
        default=-2.4737
        thinned_from=177
        keys [[0,-1.9283],[2,-1.8911],[3,-1.8879],[5,-1.8712],[6,-1.867],[7,-1.8557],[8,-1.8493],[10,-1.8452],[11,-1.8471],[12,-1.8318],[13,-1.8295],[14,-1.8462],[15,-1.8557],[17,-1.8651],[18,-1.8647],[19,-1.8739],[20,-1.8759],[22,-1.8668],[23,-1.8679],[24,-1.8582],[26,-1.8504],[27,-1.8417],[28,-1.8477],[29,-1.8435],[31,-1.8429],[32,-1.8631],[33,-1.8638],[34,-1.8575],[35,-1.8654],[37,-1.8986],[38,-1.8954],[39,-1.8823],[40,-1.8554],[41,-1.8363],[43,-1.8515],[44,-1.8723],[45,-1.9229],[46,-2.0106],[47,-2.1322],[48,-2.2846],[49,-2.4646],[50,-2.6691],[51,-2.8949],[52,-3.139],[53,-3.3981],[54,-3.6692],[55,-3.949],[56,-4.3455],[57,-4.9247],[58,-5.6206],[59,-6.3674],[60,-7.0991],[61,-7.7498],[62,-8.2536],[63,-8.6193],[64,-8.9156],[65,-9.1666],[67,-9.6283],[68,-9.887],[70,-10.4928],[72,-11.1211],[75,-11.8706],[76,-12.0539],[78,-12.6906],[79,-12.7778],[84,-13.4173],[96,-13.3945],[97,-13.3317],[99,-13.0767],[100,-12.8346],[101,-12.7828],[102,-12.7639],[105,-12.8161],[106,-12.9086],[108,-12.948],[111,-12.9197],[114,-12.9379],[135,-12.9348],[138,-12.8747],[163,-12.8943],[165,-12.9281],[167,-12.983],[168,-12.775],[169,-12.8459],[171,-13.3709],[172,-13.5103],[174,-13.729],[179,-13.7349],[181,0.5653],[182,0.4834],[183,-7.353],[184,-2.4737]]
       }
       lowerarm_l_fk_ctrl.Rotation.Z {
        type=MovieSceneScriptingFloatChannel
        default=-18.572
        thinned_from=178
        keys [[0,-21.0317],[2,-21.2188],[6,-21.3374],[8,-21.4297],[11,-21.4381],[12,-21.5156],[13,-21.5261],[15,-21.3944],[18,-21.3489],[20,-21.2914],[23,-21.3325],[27,-21.4627],[28,-21.4305],[31,-21.4537],[32,-21.3528],[34,-21.3793],[35,-21.3385],[37,-21.172],[38,-21.1861],[39,-21.2542],[41,-21.4877],[43,-21.4091],[44,-21.3038],[45,-20.9941],[46,-20.3979],[47,-19.5412],[48,-18.4496],[49,-17.1489],[50,-15.6648],[51,-14.0233],[52,-12.2499],[53,-10.3705],[54,-8.4108],[55,-6.3967],[56,-3.5245],[57,0.6944],[58,5.7615],[59,11.1785],[60,16.4469],[61,21.0684],[62,24.5446],[63,26.9903],[64,28.9644],[65,30.6343],[67,33.7317],[68,35.4941],[69,37.5946],[70,39.8837],[72,44.6724],[75,50.8219],[76,52.4041],[77,55.1919],[78,58.2213],[79,59.0838],[84,65.6315],[86,65.8043],[96,65.4371],[97,64.7585],[98,63.5297],[99,62.1121],[100,59.7463],[101,59.2086],[102,59.0228],[105,59.6419],[106,60.328],[108,60.7339],[111,60.4782],[112,60.6408],[114,60.6458],[115,60.4875],[118,60.5081],[122,60.6818],[131,60.6041],[133,60.7604],[135,60.683],[137,60.2304],[138,60.1038],[144,60.2162],[145,60.1606],[153,60.4824],[159,60.4045],[163,60.4199],[165,60.6753],[167,61.4309],[168,59.7839],[169,59.006],[170,56.6611],[171,51.9156],[172,44.73],[174,32.8867],[175,31.0887],[176,29.6861],[177,31.629],[179,39.0587],[181,46.1354],[182,50.87],[183,13.0777],[184,-18.572]]
       }
       lowerarm_l_fk_ctrl.Scale.X {type=MovieSceneScriptingFloatChannel, default=1, const [1,177]}
       lowerarm_l_fk_ctrl.Scale.Y {type=MovieSceneScriptingFloatChannel, default=1, const [1,177]}
       lowerarm_l_fk_ctrl.Scale.Z {type=MovieSceneScriptingFloatChannel, default=1, const [1,177]}
       hand_l_fk_ctrl.Location.X {
        type=MovieSceneScriptingFloatChannel
        default=-0.0003
        thinned_from=181
        keys [[0,0],[45,-0.0001],[49,-0.0011],[55,-0.0039],[56,-0.0049],[57,-0.0067],[59,-0.011],[60,-0.013],[61,-0.0144],[62,-0.0149],[63,-0.0147],[64,-0.014],[65,-0.013],[67,-0.0104],[68,-0.0089],[69,-0.0048],[70,-0.0015],[72,-0.0088],[74,0],[75,-0.0018],[76,-0.0007],[77,-0.0043],[78,-0.001],[79,-0.0012],[85,0],[165,0],[166,-0.0124],[167,0],[168,-0.0108],[169,-0.0071],[170,-0.0058],[171,-0.0036],[172,-0.0005],[173,0],[174,-0.0002],[175,0],[177,-0.0002],[178,0],[182,0],[183,-0.0044],[184,-0.0003]]
       }
       hand_l_fk_ctrl.Location.Y {
        type=MovieSceneScriptingFloatChannel
        default=0
        thinned_from=181
        keys [[0,0],[52,0.0005],[56,0.0011],[60,0.0029],[62,0.0034],[66,0.0027],[68,0.0021],[70,0.0004],[72,0.0018],[74,0],[75,0.0003],[76,0.0001],[77,0.0004],[78,0],[165,0],[166,0.0021],[167,0],[168,0.0019],[169,0.0002],[170,-0.0003],[172,0],[173,-0.431],[174,0],[175,-0.431],[176,0],[177,0],[178,-0.431],[179,0],[180,-0.431],[182,-0.431],[183,-0.0022],[184,0]]
       }
       hand_l_fk_ctrl.Location.Z {
        type=MovieSceneScriptingFloatChannel
        default=0
        thinned_from=181
        keys [[0,0],[55,-0.0006],[61,-0.0022],[63,-0.0023],[68,-0.0011],[70,-0.0001],[72,-0.0007],[74,0],[165,0],[166,-0.0035],[167,0],[168,-0.0013],[170,0.0001],[172,0],[173,0.0668],[174,0],[175,0.0668],[176,0],[177,0],[178,0.0668],[179,0],[180,0.0668],[182,0.0668],[183,0.0013],[184,0]]
       }
       hand_l_fk_ctrl.Rotation.X {
        type=MovieSceneScriptingFloatChannel
        default=4.3386
        thinned_from=181
        keys [[0,-2.7627],[1,-2.828],[2,-2.9402],[3,-3.0207],[4,-3.2284],[5,-3.164],[6,-3.4734],[7,-3.5643],[8,-3.5332],[9,-3.6426],[10,-3.5819],[11,-3.7757],[12,-3.7423],[13,-3.7282],[14,-3.622],[15,-3.469],[16,-3.4269],[17,-3.245],[18,-3.2946],[19,-3.3357],[20,-3.2772],[21,-3.2632],[22,-3.2252],[23,-3.0882],[24,-3.0782],[25,-3.1722],[26,-3.3838],[27,-3.6066],[28,-3.6092],[29,-3.7035],[30,-3.6506],[31,-3.7232],[32,-3.503],[33,-3.5396],[34,-3.542],[35,-3.559],[37,-3.1631],[38,-3.3068],[39,-3.1851],[40,-3.3586],[41,-3.36],[42,-3.2681],[43,-3.2133],[44,-3.1163],[45,-2.9311],[46,-2.67],[47,-2.3379],[48,-1.9397],[49,-1.4804],[50,-0.965],[51,-0.3984],[52,0.2145],[53,0.8687],[54,1.5592],[55,2.2812],[56,3.6869],[57,6.1411],[58,9.2075],[59,12.4495],[60,15.4309],[61,17.7152],[62,18.8661],[64,19.7111],[65,20.0726],[66,20.3611],[67,20.552],[68,20.621],[70,20.1418],[71,20.3154],[72,20.6413],[73,21.0446],[74,21.3607],[75,21.4478],[76,20.9632],[77,19.8095],[78,19.0018],[79,18.7917],[84,21.7119],[85,20.2879],[86,19.2482],[87,18.578],[88,18.1634],[89,17.9616],[90,17.8657],[96,17.831],[97,16.9835],[98,15.5864],[99,13.6949],[100,8.4104],[101,7.5938],[102,7.5892],[103,7.4994],[104,8.2404],[105,8.6371],[106,8.97],[107,9.1673],[109,9.3785],[110,9.15],[111,9.3292],[112,9.3911],[113,10.1985],[114,10.5757],[115,10.7872],[116,10.6236],[117,10.6],[118,10.739],[119,10.9822],[120,11.016],[122,11.2761],[123,11.3859],[124,11.3011],[125,11.1475],[126,11.2656],[127,11.3173],[129,11.0033],[130,11.051],[131,11.1874],[133,11.485],[134,11.5427],[135,11.403],[136,10.8097],[137,10.2387],[138,9.9741],[139,10.0368],[140,10.298],[141,10.3323],[142,10.1924],[143,10.2468],[144,10.4221],[145,10.5706],[147,11.1],[148,11.182],[149,11.3028],[151,11.0679],[152,10.9219],[153,10.7221],[154,10.4299],[156,10.2228],[158,10.2419],[159,10.2358],[160,10.4307],[161,10.4463],[162,10.3616],[164,10.3557],[165,10.9537],[166,12.1891],[167,13.7957],[168,14.0232],[169,12.152],[170,12.1971],[171,13.7164],[172,15.8681],[173,14.6393],[174,20.6703],[175,16.3146],[176,20.3119],[177,19.9011],[178,17.2037],[179,18.6481],[180,16.7825],[181,0],[182,0],[183,11.2631],[184,4.3386]]
       }
       hand_l_fk_ctrl.Rotation.Y {
        type=MovieSceneScriptingFloatChannel
        default=27.6739
        thinned_from=181
        keys [[0,22.0558],[1,22.0555],[2,22.2119],[3,22.0704],[4,22.1001],[5,22.2635],[6,22.1886],[7,22.1797],[8,22.2537],[9,22.1489],[10,22.2648],[12,22.2982],[14,22.3805],[15,22.4962],[17,22.3558],[18,22.4069],[19,22.3099],[20,22.3158],[22,22.4284],[23,22.3032],[24,22.3042],[25,22.2424],[26,22.264],[27,22.1713],[28,22.1954],[29,22.3133],[30,22.3131],[31,22.448],[33,22.4661],[34,22.5839],[35,22.4778],[36,22.531],[37,22.7258],[38,22.6223],[40,22.5964],[41,22.6273],[42,22.7577],[43,22.9762],[44,22.7833],[55,22.4731],[57,22.2884],[60,21.8833],[61,21.7933],[62,21.7589],[63,21.8238],[64,21.9962],[69,23.3974],[71,24.0174],[72,24.2756],[74,24.6265],[75,25.0037],[76,25.5],[77,26.6363],[78,27.3679],[79,27.6297],[84,31.2994],[85,31.6481],[87,31.3444],[96,31.1916],[97,31.6984],[98,31.7684],[99,32.1001],[100,32.1889],[102,31.4566],[103,31.286],[104,31.7054],[105,31.5564],[106,31.5704],[108,31.0222],[109,30.9192],[110,30.5053],[111,30.3927],[112,29.5415],[113,30.1823],[114,30.5335],[115,31.022],[116,30.7756],[117,30.7404],[118,30.7726],[119,30.6625],[122,30.5309],[123,30.6909],[125,30.5624],[126,30.6776],[127,30.6899],[129,30.5637],[130,30.5955],[131,30.7068],[133,30.6643],[134,30.7898],[137,30.4291],[139,30.3556],[140,30.5379],[141,30.5636],[142,30.4803],[143,30.4757],[145,30.7741],[146,30.7083],[147,30.9264],[151,30.7897],[152,30.689],[153,30.5101],[154,30.4163],[160,30.5108],[161,30.4616],[162,30.2788],[163,30.3101],[164,30.1139],[165,30.011],[166,28.9768],[167,25.7468],[168,25.4921],[169,25.6904],[170,25.0277],[171,25.3603],[172,24.9647],[173,23.4231],[174,24.9375],[175,23.9173],[176,25.9606],[177,26.5348],[178,24.2868],[179,27.8052],[180,26.1318],[181,0],[182,0],[183,26.5472],[184,27.6739]]
       }
       hand_l_fk_ctrl.Rotation.Z {
        type=MovieSceneScriptingFloatChannel
        default=-23.7898
        thinned_from=181
        keys [[0,-17.1358],[1,-17.3057],[3,-17.3418],[4,-17.436],[5,-17.4498],[6,-17.7574],[8,-17.819],[9,-17.8929],[10,-17.9057],[11,-18.0163],[12,-18.0649],[13,-17.9688],[14,-18.0342],[15,-18.0521],[16,-17.9448],[17,-17.7744],[18,-17.7553],[20,-17.6029],[21,-17.6763],[22,-17.6446],[23,-17.513],[24,-17.4577],[25,-17.6389],[26,-17.8625],[27,-17.9961],[28,-17.9579],[29,-18.2426],[31,-18.1536],[32,-18.2707],[33,-18.1644],[34,-18.1894],[35,-18.2942],[36,-18.3366],[37,-18.1869],[38,-18.1322],[39,-18.19],[40,-18.166],[41,-18.2026],[42,-18.3032],[43,-18.297],[44,-18.1438],[45,-17.8921],[46,-17.5966],[47,-17.2577],[48,-16.8754],[49,-16.4501],[50,-15.9817],[51,-15.4707],[52,-14.9171],[53,-14.3211],[54,-13.6829],[55,-13.0028],[56,-11.7165],[57,-9.5272],[58,-6.8362],[59,-4.0447],[60,-1.5539],[61,0.2348],[62,0.9202],[63,0.8614],[64,0.699],[65,0.4537],[66,0.1462],[67,-0.2028],[68,-0.5726],[69,-1.3355],[70,-2.293],[71,-3.0089],[72,-3.6213],[73,-4.0436],[74,-4.5144],[75,-5.5781],[76,-7.0537],[77,-10.0525],[78,-12.8013],[79,-13.3352],[84,-18.0368],[85,-19.8831],[86,-20.6599],[87,-21.2314],[88,-21.5084],[90,-21.6504],[96,-21.683],[97,-22.5639],[98,-22.6087],[99,-22.5857],[100,-23.9039],[101,-23.4257],[102,-22.3681],[103,-22.0875],[104,-21.501],[105,-21.1921],[106,-21.0968],[108,-20.5889],[109,-20.2871],[110,-20.026],[111,-20.0049],[112,-20.9357],[113,-20.1283],[115,-17.5965],[116,-17.5335],[117,-17.5159],[118,-17.4342],[119,-17.2993],[120,-17.2408],[121,-17.1323],[122,-16.9342],[123,-17.0094],[124,-16.9033],[125,-17.0057],[126,-17.0727],[127,-17.0739],[128,-16.8631],[130,-16.7654],[131,-16.827],[132,-16.6748],[133,-16.5739],[134,-16.5953],[135,-16.4723],[137,-16.7906],[138,-16.7138],[139,-16.6931],[140,-16.7154],[141,-16.8617],[143,-16.7549],[144,-16.5984],[145,-16.5279],[146,-16.2785],[147,-16.252],[148,-16.2774],[149,-16.0239],[150,-16.0048],[151,-16.0349],[152,-16.1247],[154,-16.13],[155,-16.0522],[156,-16.0544],[157,-16.0172],[158,-15.8695],[159,-15.7636],[160,-15.5744],[161,-15.7356],[163,-15.6557],[164,-15.4024],[165,-14.8724],[166,-12.8741],[167,-9.9676],[168,-12.5479],[169,-16.7464],[170,-16.8992],[171,-18.7765],[172,-17.6892],[173,-14.7362],[174,-11.9709],[175,-11.0489],[176,-13.314],[177,-14.1229],[178,-8.9979],[179,-17.7947],[180,-12.2491],[181,0],[182,0],[183,-19.1881],[184,-23.7898]]
       }
       hand_l_fk_ctrl.Scale.X {type=MovieSceneScriptingFloatChannel, default=1, const [1,181]}
       hand_l_fk_ctrl.Scale.Y {type=MovieSceneScriptingFloatChannel, default=1, const [1,181]}
       hand_l_fk_ctrl.Scale.Z {type=MovieSceneScriptingFloatChannel, default=1, const [1,181]}
       thumb_01_l_ctrl.Location.X {
        type=MovieSceneScriptingFloatChannel
        default=0
        thinned_from=185
        keys [[0,0],[56,-0.0003],[62,-0.001],[70,-0.0001],[72,-0.0003],[74,0],[77,-0.0002],[82,-0.0001],[165,0],[166,-0.0016],[167,0],[168,-0.0015],[169,-0.001],[171,-0.0005],[172,-0.0001],[179,0],[180,-0.0002],[181,-0.0009],[182,-0.0009],[183,0],[184,0]]
       }
       thumb_01_l_ctrl.Location.Y {
        type=MovieSceneScriptingFloatChannel
        default=0
        thinned_from=185
        keys [[0,0],[55,-0.0004],[61,-0.0016],[65,-0.0013],[68,-0.0008],[70,-0.0001],[72,-0.0008],[74,0],[77,-0.0002],[165,0],[166,-0.0009],[167,0],[168,-0.0004],[171,-0.0003],[172,0],[180,-0.0001],[181,-0.0007],[182,-0.0011],[183,-0.0003],[184,0]]
       }
       thumb_01_l_ctrl.Location.Z {
        type=MovieSceneScriptingFloatChannel
        default=0
        thinned_from=185
        keys [[0,0],[56,-0.0004],[62,-0.0011],[68,-0.0007],[70,-0.0001],[72,-0.0007],[74,0],[75,-0.0002],[76,-0.0001],[77,-0.0006],[78,-0.0001],[82,-0.0003],[83,0],[165,0],[166,-0.0005],[167,0],[168,-0.0001],[171,0.0002],[180,0],[181,-0.0005],[182,-0.0005],[183,0.0001],[184,0]]
       }
       thumb_01_l_ctrl.Rotation.X {
        type=MovieSceneScriptingFloatChannel
        default=-5.7048
        thinned_from=185
        keys [[0,-3.0963],[1,-3.0189],[2,-3.083],[3,-3.0107],[4,-3.0061],[5,-2.9909],[6,-2.9571],[7,-2.8393],[8,-2.9729],[9,-2.8593],[10,-2.9956],[11,-2.9158],[12,-2.9075],[13,-2.8923],[14,-2.8164],[15,-2.8202],[16,-2.7136],[17,-2.7728],[18,-2.8789],[19,-2.856],[20,-2.8508],[21,-2.8735],[22,-2.9759],[23,-2.885],[24,-2.9052],[25,-2.9479],[26,-2.9451],[27,-2.8499],[28,-3.0697],[29,-2.9383],[30,-2.9528],[31,-2.9104],[33,-2.935],[34,-2.9402],[35,-2.907],[36,-2.8889],[37,-2.8891],[38,-2.9361],[40,-2.8794],[41,-2.7726],[42,-2.8194],[43,-2.6003],[44,-2.8851],[55,-2.8533],[56,-2.8446],[61,-2.7514],[66,-2.723],[70,-2.7218],[71,-2.7424],[72,-2.7865],[73,-2.936],[74,-3.0612],[75,-3.0569],[76,-3.2056],[77,-3.338],[78,-3.2576],[79,-3.2815],[81,-3.3868],[82,-3.4107],[83,-2.7006],[84,-1.8794],[85,-0.8986],[86,0.0269],[87,0.2931],[88,0.5101],[89,0.57],[90,0.6039],[91,0.6223],[92,0.6307],[96,0.631],[97,-0.1625],[98,-0.621],[99,-0.8977],[100,-0.8019],[101,-1.2461],[102,-1.6534],[103,-1.7384],[104,-1.9539],[105,-1.8219],[106,-1.5933],[107,-1.4512],[108,-1.3856],[109,-1.4186],[110,-1.3985],[111,-1.3617],[112,-0.3647],[113,-1.1378],[114,-1.9908],[115,-3.0326],[116,-3.1597],[117,-3.0049],[118,-2.8156],[119,-2.9144],[120,-2.9377],[121,-2.9856],[122,-2.9346],[124,-2.8472],[125,-2.7351],[126,-2.8913],[127,-2.8429],[128,-2.8599],[129,-2.9372],[130,-2.8734],[131,-2.7262],[132,-2.6642],[133,-2.7088],[134,-2.826],[135,-2.9065],[136,-3.0099],[137,-2.8817],[138,-2.9695],[139,-2.8598],[140,-2.8877],[141,-2.892],[142,-2.7342],[144,-2.9589],[145,-3.8598],[146,-3.0089],[147,-3.9511],[148,-3.2314],[149,-3.0619],[150,-3.1124],[151,-3.0381],[152,-3.0115],[153,-3.0925],[154,-3.0851],[155,-3.0888],[156,-3.0494],[157,-3.0924],[158,-3.1192],[159,-3.2945],[160,-3.3066],[161,-3.1319],[162,-3.1218],[163,-3.2281],[164,-3.2611],[165,-3.341],[166,-3.7743],[167,-4.0393],[168,-3.9411],[169,-3.5788],[170,-2.693],[171,-1.7664],[172,-1.5179],[173,-2.0899],[174,-2.5736],[175,-3.0306],[176,-3.5983],[177,-3.4597],[178,-3.2245],[179,-2.9583],[180,-2.7543],[181,-2.6093],[182,-2.3468],[183,-5.2271],[184,-5.7048]]
       }
       thumb_01_l_ctrl.Rotation.Y {
        type=MovieSceneScriptingFloatChannel
        default=11.1309
        thinned_from=185
        keys [[0,12.3316],[1,12.4292],[2,12.3235],[3,12.4542],[4,12.4028],[5,12.2854],[6,12.3969],[7,12.3843],[8,12.3315],[9,12.46],[10,12.3855],[11,12.2506],[12,12.3385],[13,12.2718],[15,12.3752],[16,12.4081],[17,12.3588],[18,12.284],[20,12.2647],[21,12.2812],[22,12.2445],[23,12.3539],[24,12.2595],[25,12.4099],[26,12.3628],[27,12.3766],[28,12.2013],[29,12.2938],[30,12.3228],[31,12.1847],[32,12.2666],[33,12.2702],[34,12.1962],[36,12.2875],[38,12.1689],[40,12.3252],[41,12.3209],[42,12.2835],[43,12.5264],[44,12.3036],[51,12.4316],[55,12.5711],[56,12.6397],[57,12.7579],[60,13.1909],[61,13.2883],[62,13.3256],[63,13.2915],[67,12.8992],[68,12.8651],[74,13.011],[76,13.2114],[77,13.0103],[78,13.3334],[82,13.3628],[83,13.2381],[84,13.3442],[85,13.1201],[86,13.2201],[87,13.2017],[96,13.2674],[97,13.4849],[99,14.0555],[100,14.066],[101,14.2484],[103,14.3415],[104,14.1386],[106,14.0639],[109,14.2085],[110,14.2103],[111,14.3053],[112,15.2796],[113,14.9605],[114,14.2594],[115,13.2059],[116,13.2427],[122,13.1954],[123,13.2349],[124,13.2461],[125,13.2174],[129,13.194],[130,13.2184],[132,13.1543],[133,13.1427],[135,13.1893],[136,13.1371],[137,13.2894],[138,13.2297],[139,13.3175],[140,13.2263],[142,13.3032],[143,13.2498],[145,13.2917],[146,13.3501],[147,13.2669],[148,13.3392],[149,13.3386],[150,13.2394],[151,13.265],[152,13.2396],[153,13.2453],[154,13.3125],[155,13.3071],[157,13.222],[160,13.3695],[161,13.5177],[162,13.619],[163,13.6814],[164,13.7098],[165,13.7969],[166,13.5831],[167,13.4498],[168,12.9154],[169,12.8328],[170,12.647],[171,12.2922],[172,12.349],[173,12.5054],[174,12.4463],[175,12.3119],[176,12.0656],[177,11.8889],[178,11.9612],[179,11.9653],[180,11.8965],[181,11.679],[182,11.4244],[183,11.0167],[184,11.1309]]
       }
       thumb_01_l_ctrl.Rotation.Z {
        type=MovieSceneScriptingFloatChannel
        default=-20.9233
        thinned_from=185
        keys [[0,-19.689],[1,-19.5679],[2,-19.7546],[3,-19.6125],[4,-19.685],[5,-19.8011],[6,-19.6875],[7,-19.6962],[8,-19.7838],[9,-19.6456],[10,-19.7516],[11,-19.9043],[12,-19.875],[13,-19.9891],[14,-19.8504],[17,-19.8062],[19,-19.868],[20,-19.934],[21,-19.8723],[22,-19.8931],[23,-19.7932],[24,-19.843],[25,-19.7197],[26,-19.7445],[27,-19.6863],[28,-19.8063],[29,-19.795],[30,-19.739],[31,-19.9618],[32,-19.7962],[34,-19.8934],[36,-19.8617],[37,-19.9546],[38,-19.9832],[39,-19.8991],[41,-19.8638],[42,-19.8795],[43,-19.8348],[57,-19.9094],[62,-20.0953],[63,-20.1913],[67,-20.9621],[68,-21.0698],[72,-21.1817],[74,-21.3434],[75,-21.3094],[76,-21.3225],[77,-22.0009],[78,-21.8349],[82,-21.8844],[83,-22.8067],[84,-22.7114],[86,-22.2157],[89,-21.9009],[96,-21.7725],[97,-21.6092],[98,-21.3234],[99,-21.0976],[100,-20.6851],[101,-20.5108],[102,-20.4001],[103,-20.3882],[104,-20.522],[106,-20.5812],[107,-20.5339],[108,-20.4386],[110,-20.3702],[111,-20.2453],[112,-18.1582],[113,-18.6558],[114,-20.1129],[115,-22.3526],[116,-22.3116],[122,-22.4965],[123,-22.4673],[125,-22.5187],[126,-22.4657],[129,-22.4714],[130,-22.3762],[133,-22.4209],[134,-22.4728],[135,-22.4233],[136,-22.4666],[137,-22.1767],[138,-22.3532],[139,-22.1803],[140,-22.3452],[141,-22.1918],[142,-22.1432],[143,-22.3268],[145,-22.3989],[146,-22.1452],[147,-22.3237],[148,-22.0541],[149,-22.0966],[150,-22.2892],[151,-22.212],[153,-22.248],[154,-22.1448],[157,-22.1658],[158,-22.0532],[160,-21.8969],[161,-21.593],[163,-21.277],[164,-21.1679],[165,-21.0057],[166,-21.1896],[167,-21.1226],[168,-21.3772],[169,-20.989],[170,-20.7929],[171,-20.6435],[173,-19.5876],[174,-19.8091],[175,-19.8731],[176,-20.2549],[177,-20.4364],[178,-20.1817],[179,-20.1205],[180,-20.4996],[181,-20.9751],[182,-21.116],[183,-21.3289],[184,-20.9233]]
       }
       thumb_01_l_ctrl.Scale.X {type=MovieSceneScriptingFloatChannel, default=1, const [1,185]}
       thumb_01_l_ctrl.Scale.Y {type=MovieSceneScriptingFloatChannel, default=1, const [1,185]}
       thumb_01_l_ctrl.Scale.Z {type=MovieSceneScriptingFloatChannel, default=1, const [1,185]}
       thumb_02_l_ctrl.Location.X {
        type=MovieSceneScriptingFloatChannel
        default=0
        thinned_from=185
        keys [[0,0],[56,-0.0005],[61,-0.0014],[68,-0.0014],[70,-0.0002],[72,-0.0013],[74,0],[75,-0.0005],[76,-0.0002],[77,-0.0014],[78,-0.0004],[82,-0.0009],[83,-0.0003],[84,-0.0003],[85,0],[165,0],[166,-0.0027],[167,0],[168,-0.0028],[169,-0.0018],[171,-0.0013],[172,-0.0001],[179,0],[180,-0.0003],[181,-0.0012],[182,0.0001],[183,-0.0006],[184,0]]
       }
       thumb_02_l_ctrl.Location.Y {
        type=MovieSceneScriptingFloatChannel
        default=0
        thinned_from=185
        keys [[0,0],[56,-0.0003],[62,-0.001],[68,-0.0004],[70,-0.0001],[72,-0.0003],[74,0],[76,-0.0001],[77,-0.0002],[82,-0.0003],[83,0.0001],[165,0],[166,-0.0021],[167,0],[168,-0.0021],[169,-0.0015],[171,-0.0008],[172,-0.0001],[179,0],[180,-0.0002],[181,-0.0012],[182,-0.001],[183,0.0002],[184,0]]
       }
       thumb_02_l_ctrl.Location.Z {
        type=MovieSceneScriptingFloatChannel
        default=0
        thinned_from=185
        keys [[0,0],[55,-0.0006],[61,-0.0022],[64,-0.0021],[68,-0.0012],[70,-0.0002],[72,-0.0011],[74,0],[75,-0.0002],[76,-0.0001],[77,-0.0003],[78,-0.0001],[82,-0.0002],[84,0],[165,0],[166,-0.0014],[167,0],[168,-0.0007],[169,-0.0001],[171,0.0003],[172,0],[180,-0.0001],[181,-0.0012],[183,0.0004],[184,0]]
       }
       thumb_02_l_ctrl.Rotation.X {
        type=MovieSceneScriptingFloatChannel
        default=-3.5612
        thinned_from=185
        keys [[0,-3.3555],[1,-3.3496],[2,-3.3043],[3,-3.3219],[4,-3.3201],[5,-3.2898],[6,-3.2664],[7,-3.2506],[8,-3.2728],[9,-3.2847],[10,-3.3154],[11,-3.2221],[12,-3.2395],[14,-3.2362],[15,-3.2251],[16,-3.1835],[18,-3.25],[20,-3.2051],[21,-3.2306],[22,-3.2885],[23,-3.244],[26,-3.334],[27,-3.2616],[28,-3.3028],[29,-3.2579],[30,-3.2811],[31,-3.2239],[32,-3.2512],[33,-3.2415],[34,-3.2724],[35,-3.2614],[36,-3.2613],[37,-3.2532],[38,-3.2609],[39,-3.2437],[40,-3.24],[41,-3.2228],[42,-3.2525],[43,-3.211],[44,-3.2519],[45,-3.2464],[46,-3.2305],[47,-3.2053],[48,-3.1719],[50,-3.0842],[52,-2.9759],[55,-2.7928],[56,-2.673],[57,-2.4641],[58,-2.2042],[59,-1.9312],[60,-1.683],[61,-1.4976],[62,-1.413],[64,-1.3714],[65,-1.3537],[66,-1.3395],[67,-1.3301],[68,-1.3267],[69,-1.3407],[70,-1.3615],[72,-1.3813],[73,-1.4336],[74,-1.5097],[75,-1.5692],[76,-1.6682],[77,-1.7909],[78,-1.762],[79,-1.787],[81,-1.8974],[82,-1.9225],[83,-1.8179],[84,-1.8243],[85,-1.628],[86,-1.4546],[87,-1.3988],[88,-1.3539],[90,-1.3357],[91,-1.33],[96,-1.3284],[97,-1.4693],[98,-1.5665],[99,-1.6376],[100,-1.679],[101,-1.8417],[102,-1.9759],[103,-2.0036],[104,-2.1098],[106,-1.9878],[107,-1.9385],[108,-1.9113],[109,-1.9133],[110,-1.9045],[112,-1.9167],[113,-1.9751],[114,-1.9975],[115,-1.9886],[116,-2.0306],[117,-2.0041],[118,-1.9387],[119,-1.9733],[120,-1.9734],[121,-1.9814],[122,-1.9485],[125,-1.8909],[126,-1.9388],[127,-1.9201],[129,-1.9362],[130,-1.9088],[131,-1.8608],[132,-1.847],[133,-1.8525],[135,-1.9111],[136,-1.9548],[137,-1.9282],[138,-1.9331],[140,-1.8859],[141,-1.9098],[142,-1.8649],[144,-1.8846],[145,-2.204],[146,-1.9067],[147,-2.2343],[148,-1.988],[149,-1.9184],[150,-1.9308],[151,-1.8984],[152,-1.8959],[153,-1.9302],[154,-1.9201],[155,-1.9204],[156,-1.9311],[157,-1.9361],[158,-1.9489],[159,-1.9928],[160,-2.0072],[161,-2.0023],[162,-2.0391],[163,-2.0966],[164,-2.1211],[165,-2.1697],[166,-2.1598],[167,-2.064],[168,-2.0212],[169,-2.0259],[170,-2.0961],[171,-2.1854],[172,-2.3583],[173,-2.6213],[174,-2.8042],[175,-2.9784],[176,-3.131],[177,-3.1275],[179,-3.0056],[180,-2.9635],[181,-2.8869],[182,-2.7387],[183,-3.3248],[184,-3.5612]]
       }
       thumb_02_l_ctrl.Rotation.Y {
        type=MovieSceneScriptingFloatChannel
        default=11.9115
        thinned_from=185
        keys [[0,11.0131],[1,10.9424],[3,10.8641],[4,10.8837],[7,10.6582],[8,10.7855],[9,10.7621],[10,10.9003],[11,10.6392],[12,10.6981],[15,10.6218],[16,10.493],[17,10.5947],[18,10.7315],[19,10.6369],[20,10.5929],[21,10.6726],[22,10.8705],[23,10.6741],[24,10.7639],[25,10.8132],[26,10.9091],[27,10.6657],[28,10.8781],[29,10.6898],[30,10.7972],[31,10.6463],[32,10.7208],[33,10.6861],[34,10.8019],[35,10.7476],[38,10.7771],[40,10.6818],[41,10.6043],[42,10.6984],[43,10.5361],[44,10.7434],[45,10.7215],[46,10.6585],[47,10.5585],[48,10.4256],[49,10.2639],[50,10.0776],[52,9.6473],[55,8.9198],[56,8.4451],[57,7.6189],[58,6.5916],[59,5.514],[60,4.5366],[61,3.81],[62,3.485],[64,3.3423],[65,3.2812],[66,3.2325],[67,3.2003],[68,3.1886],[69,3.236],[70,3.3177],[71,3.3802],[72,3.4629],[73,3.6696],[74,3.9382],[75,4.1613],[76,4.5379],[77,5.0101],[78,4.8776],[79,4.9634],[81,5.3413],[82,5.4272],[83,4.9829],[84,4.6337],[85,3.9145],[86,3.2309],[87,3.025],[88,2.8515],[89,2.8114],[90,2.7804],[91,2.759],[92,2.7529],[96,2.7529],[97,3.2738],[98,3.628],[99,3.8998],[100,4.0987],[101,4.6692],[102,5.1455],[103,5.2474],[104,5.6599],[105,5.4583],[106,5.2426],[107,5.05],[108,4.9464],[109,4.9577],[110,4.9243],[111,4.9328],[112,4.9801],[113,5.2935],[114,5.4201],[115,5.521],[116,5.654],[117,5.5511],[118,5.3095],[119,5.4373],[120,5.4234],[121,5.4534],[122,5.33],[123,5.2655],[124,5.1725],[125,5.1049],[126,5.2774],[127,5.2179],[128,5.2405],[129,5.2759],[130,5.1761],[131,5.0026],[132,4.9405],[133,4.9621],[134,5.0875],[135,5.1891],[136,5.3559],[137,5.2542],[138,5.2712],[139,5.1839],[140,5.1217],[141,5.2325],[142,5.0491],[143,5.1154],[144,5.1519],[145,6.3687],[146,5.2397],[147,6.4983],[148,5.5738],[149,5.2886],[150,5.3518],[151,5.23],[152,5.2253],[153,5.3503],[154,5.2785],[155,5.2802],[156,5.3273],[157,5.3387],[158,5.3653],[159,5.5553],[160,5.5918],[161,5.5185],[162,5.6037],[163,5.7954],[164,5.8656],[165,6.0184],[166,6.101],[167,5.897],[168,5.9284],[169,6.0193],[170,6.4144],[171,7.197],[172,7.8882],[173,8.9426],[175,10.4287],[176,10.9565],[177,10.9148],[179,10.314],[180,10.1393],[181,9.7919],[182,8.7228],[183,11.5073],[184,11.9115]]
       }
       thumb_02_l_ctrl.Rotation.Z {
        type=MovieSceneScriptingFloatChannel
        default=29.815
        thinned_from=185
        keys [[0,28.8105],[1,28.6151],[2,28.974],[3,28.6518],[5,28.865],[6,28.8364],[7,28.73],[8,28.9411],[9,28.712],[11,28.9975],[13,29.0001],[14,28.8989],[17,28.8968],[18,29.0089],[19,28.9123],[20,29.0275],[22,29.0686],[23,28.867],[24,28.8402],[25,28.6927],[27,28.6263],[28,28.9297],[29,28.7605],[31,29.0027],[33,28.9419],[34,29.0048],[35,28.9349],[36,28.9731],[37,29.1049],[38,29.0483],[39,29.0592],[41,28.8599],[42,28.8567],[43,28.7461],[44,29.0307],[47,28.9105],[52,28.3186],[55,27.8459],[56,27.5421],[57,27.0186],[59,25.701],[60,25.101],[61,24.6698],[62,24.5045],[70,24.654],[72,25.1326],[74,25.1378],[75,25.1883],[76,25.2933],[77,25.4885],[78,25.3223],[82,25.2253],[83,24.7826],[84,21.9863],[85,21.5993],[86,20.8187],[87,20.7119],[88,20.5311],[91,20.426],[96,20.4205],[97,20.6884],[98,20.8232],[99,21.0986],[100,21.7104],[101,21.7276],[102,21.8497],[103,21.9142],[104,22.359],[105,22.4213],[106,22.3656],[107,22.1639],[108,22.0761],[109,22.116],[110,22.0877],[111,21.9172],[112,22.2093],[113,23.1411],[114,23.5029],[115,24.3867],[118,24.1463],[119,24.1931],[120,24.1001],[121,24.1143],[122,24.0565],[123,24.1103],[124,23.9054],[127,23.9681],[128,23.9335],[129,23.9816],[131,23.9164],[132,23.8234],[133,23.8403],[134,24.001],[135,23.9874],[136,24.0817],[137,24.0238],[139,24.0131],[140,24.1249],[141,24.3039],[142,24.1347],[143,24.3227],[144,24.3521],[145,25.0086],[146,24.4189],[147,25.1538],[148,24.7271],[149,24.4683],[150,24.5901],[151,24.5393],[153,24.5952],[154,24.3663],[155,24.3712],[156,24.4324],[157,24.3929],[158,24.2786],[159,24.511],[160,24.4269],[161,24.0628],[162,23.7812],[163,23.735],[165,23.544],[167,25.0701],[168,26.0603],[169,26.4167],[170,27.1078],[171,28.9685],[172,29.411],[173,29.9784],[175,30.9653],[176,31.1497],[177,30.9852],[178,30.6882],[179,30.3098],[180,30.1867],[181,29.8146],[182,27.5921],[183,30.7899],[184,29.815]]
       }
       thumb_02_l_ctrl.Scale.X {type=MovieSceneScriptingFloatChannel, default=1, const [1,185]}
       thumb_02_l_ctrl.Scale.Y {type=MovieSceneScriptingFloatChannel, default=1, const [1,185]}
       thumb_02_l_ctrl.Scale.Z {type=MovieSceneScriptingFloatChannel, default=1, const [1,185]}
       thumb_03_l_ctrl.Location.X {
        type=MovieSceneScriptingFloatChannel
        default=0
        thinned_from=185
        keys [[0,0],[68,-0.0006],[70,-0.0001],[72,-0.0006],[74,0],[75,-0.0003],[76,-0.0001],[77,-0.0006],[78,-0.0002],[82,-0.0004],[83,-0.0001],[165,0],[166,-0.0011],[167,0],[168,-0.0013],[169,-0.0009],[171,-0.0004],[172,-0.0001],[179,0],[180,-0.0001],[181,-0.0007],[182,-0.0003],[184,0]]
       }
       thumb_03_l_ctrl.Location.Y {
        type=MovieSceneScriptingFloatChannel
        default=0
        thinned_from=185
        keys [[0,0],[77,0.0001],[165,0],[166,-0.0003],[167,0],[168,-0.0002],[180,0],[182,-0.0003],[183,0.0002],[184,0]]
       }
       thumb_03_l_ctrl.Location.Z {
        type=MovieSceneScriptingFloatChannel
        default=0
        thinned_from=185
        keys [[0,0],[57,-0.0004],[63,-0.0009],[68,-0.0005],[70,-0.0001],[72,-0.0004],[74,0],[82,-0.0001],[84,0.0001],[165,0],[166,-0.0004],[167,0],[168,-0.0002],[171,0.0001],[180,0],[181,-0.0004],[182,-0.0004],[183,0],[184,0]]
       }
       thumb_03_l_ctrl.Rotation.X {
        type=MovieSceneScriptingFloatChannel
        default=1.9827
        thinned_from=185
        keys [[0,1.8232],[1,1.7846],[2,1.9073],[3,1.8309],[4,1.8452],[6,1.9086],[7,1.8696],[8,1.9168],[9,1.8482],[10,1.865],[11,1.9634],[13,1.9638],[14,1.9129],[15,1.9457],[17,1.9514],[18,1.9206],[19,1.9416],[20,1.976],[21,1.936],[22,1.9169],[23,1.9105],[24,1.8581],[25,1.8392],[26,1.7997],[27,1.8662],[28,1.9016],[29,1.904],[30,1.8991],[31,1.9478],[32,1.936],[33,1.9391],[34,1.9124],[35,1.9153],[36,1.9305],[37,1.9339],[38,1.9128],[39,1.9422],[40,1.9346],[41,1.9174],[42,1.9213],[43,1.8919],[44,1.9296],[47,1.9223],[55,1.8583],[56,1.84],[57,1.8085],[59,1.7292],[60,1.6931],[61,1.6672],[62,1.6572],[63,1.6648],[64,1.6839],[66,1.7336],[67,1.7526],[68,1.7603],[70,1.7535],[72,1.8274],[74,1.801],[76,1.8004],[77,1.7774],[78,1.9576],[79,1.9284],[81,1.7999],[82,1.7707],[83,1.9537],[84,1.2604],[85,1.061],[86,0.864],[87,0.8139],[88,0.7707],[89,0.751],[90,0.7426],[92,0.7369],[96,0.7358],[97,0.7884],[98,0.7327],[99,0.7074],[100,0.7104],[102,0.7018],[103,0.7016],[104,0.6972],[105,0.7165],[106,0.6985],[107,0.7046],[108,0.7091],[109,0.6969],[110,0.6918],[111,0.6792],[112,0.3323],[113,0.518],[114,0.8858],[115,1.6875],[116,1.659],[117,1.6509],[118,1.6349],[119,1.6421],[121,1.6475],[122,1.6396],[123,1.6396],[124,1.6145],[125,1.6224],[127,1.6285],[128,1.6213],[129,1.626],[130,1.6265],[131,1.6177],[132,1.6017],[133,1.6103],[134,1.6451],[135,1.638],[136,1.6546],[137,1.5689],[138,1.6131],[139,1.581],[140,1.6568],[141,1.6347],[142,1.5919],[143,1.6692],[144,1.6693],[145,1.8616],[146,1.6285],[147,1.9134],[148,1.7103],[149,1.6556],[150,1.7124],[151,1.7044],[152,1.6874],[153,1.729],[154,1.6388],[155,1.635],[156,1.6445],[157,1.6393],[158,1.5948],[159,1.651],[160,1.6046],[161,1.4425],[162,1.3372],[163,1.2845],[164,1.2408],[165,1.1866],[166,1.3837],[167,1.5892],[168,1.9358],[169,1.926],[170,1.9899],[171,2.1676],[172,2.0661],[173,2.2704],[174,2.353],[175,2.3599],[176,2.4062],[177,2.3235],[178,2.3056],[179,2.1714],[180,2.1443],[181,2.1772],[182,1.645],[183,2.3573],[184,1.9827]]
       }
       thumb_03_l_ctrl.Rotation.Y {
        type=MovieSceneScriptingFloatChannel
        default=8.7195
        thinned_from=185
        keys [[0,8.363],[1,8.2744],[2,8.5528],[3,8.3805],[4,8.4131],[6,8.5556],[7,8.4683],[8,8.5739],[9,8.42],[10,8.458],[11,8.6767],[13,8.6775],[14,8.5651],[15,8.6377],[17,8.6503],[18,8.5823],[19,8.6287],[20,8.7043],[21,8.6163],[22,8.574],[23,8.5599],[24,8.4423],[25,8.3994],[26,8.309],[27,8.4607],[28,8.54],[30,8.5345],[31,8.6424],[32,8.6163],[33,8.6233],[34,8.5642],[35,8.5705],[36,8.6044],[37,8.6118],[38,8.565],[39,8.63],[40,8.6133],[41,8.5753],[42,8.584],[43,8.5185],[44,8.6022],[55,8.4375],[56,8.3953],[57,8.3225],[60,8.0561],[61,7.9961],[62,7.9732],[63,7.9913],[67,8.1996],[68,8.2177],[70,8.2018],[72,8.3728],[74,8.312],[76,8.3108],[77,8.2582],[78,8.6654],[79,8.5999],[81,8.3114],[82,8.2459],[83,8.6646],[84,6.8975],[86,5.6846],[87,5.5155],[88,5.3687],[89,5.2995],[90,5.2695],[93,5.2457],[96,5.2447],[97,5.4391],[98,5.2328],[99,5.1364],[100,5.1483],[102,5.1149],[103,5.1144],[104,5.0972],[105,5.1715],[106,5.1022],[108,5.1431],[109,5.096],[110,5.0761],[111,5.0303],[112,3.4233],[113,4.3513],[114,5.7854],[115,8.0457],[116,7.9775],[117,7.9577],[118,7.919],[121,7.9494],[122,7.9303],[123,7.9304],[124,7.8691],[127,7.9033],[128,7.8857],[130,7.8984],[131,7.877],[132,7.8376],[133,7.8587],[134,7.9436],[135,7.9264],[136,7.9667],[137,7.7562],[138,7.8657],[139,7.7864],[140,7.9718],[141,7.9183],[142,7.8133],[143,8.0019],[144,8.0021],[145,8.4493],[146,7.9025],[147,8.5656],[148,8.0998],[149,7.9691],[150,8.1051],[151,8.0861],[152,8.0455],[153,8.1445],[154,7.9284],[155,7.9192],[156,7.9424],[157,7.9297],[158,7.8205],[159,7.9581],[160,7.8448],[161,7.4339],[162,7.1527],[163,7.0074],[164,6.8846],[165,6.7285],[167,7.8067],[168,8.6118],[169,8.5951],[170,8.7331],[171,9.1095],[172,8.8987],[173,9.3196],[174,9.4841],[175,9.4976],[176,9.588],[177,9.4254],[178,9.3905],[179,9.1185],[180,9.0626],[181,9.1299],[182,7.9454],[183,9.492],[184,8.7195]]
       }
       thumb_03_l_ctrl.Rotation.Z {
        type=MovieSceneScriptingFloatChannel
        default=23.2654
        thinned_from=185
        keys [[0,22.2422],[1,21.9895],[2,22.7857],[3,22.2922],[4,22.3853],[6,22.7938],[7,22.5435],[8,22.8462],[9,22.4052],[10,22.5138],[11,23.1418],[13,23.1441],[14,22.8211],[15,23.0297],[17,23.066],[18,22.8705],[19,23.0039],[20,23.2214],[21,22.9682],[22,22.8465],[23,22.806],[24,22.4688],[25,22.3463],[26,22.0882],[27,22.5213],[28,22.749],[29,22.7641],[30,22.7331],[31,23.0433],[32,22.9681],[33,22.9883],[34,22.8184],[35,22.8365],[36,22.9338],[37,22.955],[38,22.8208],[39,23.0074],[40,22.9594],[41,22.8502],[42,22.8752],[43,22.6871],[44,22.9274],[55,22.4587],[56,22.3385],[57,22.1314],[60,21.3728],[61,21.2023],[62,21.1369],[63,21.188],[67,21.7766],[68,21.8278],[70,21.7827],[72,22.2691],[74,22.0967],[76,22.0932],[77,21.9426],[78,23.1083],[79,22.9204],[81,22.0934],[82,21.9055],[83,23.0939],[84,18.1384],[86,14.833],[87,14.3764],[88,13.9801],[89,13.7937],[90,13.7131],[93,13.6492],[96,13.6467],[97,14.1669],[98,13.6155],[99,13.3581],[100,13.3894],[102,13.3002],[103,13.2988],[104,13.2529],[105,13.4513],[106,13.2663],[108,13.3755],[109,13.2496],[110,13.1965],[111,13.0741],[112,8.8356],[113,11.2722],[114,15.0985],[115,21.3399],[116,21.1468],[117,21.0909],[118,20.9816],[121,21.0675],[122,21.0135],[123,21.0138],[124,20.8407],[127,20.9373],[128,20.8877],[130,20.9234],[131,20.863],[132,20.752],[133,20.8114],[134,21.0511],[135,21.0024],[136,21.1169],[137,20.5229],[138,20.8311],[139,20.6078],[140,21.131],[141,20.9795],[142,20.6836],[143,21.216],[144,21.2165],[145,22.488],[146,20.934],[147,22.8215],[148,21.4929],[149,21.1229],[150,21.5082],[151,21.4543],[152,21.3394],[153,21.6205],[154,21.008],[155,20.9822],[156,21.0475],[157,21.0118],[158,20.7039],[159,21.0919],[160,20.772],[161,19.6201],[162,18.8382],[163,18.436],[164,18.097],[165,17.6672],[167,20.665],[168,22.9563],[169,22.9069],[170,23.3054],[171,24.3977],[172,23.7829],[173,25.0116],[174,25.4955],[175,25.5352],[176,25.8026],[177,25.3228],[178,25.2195],[179,24.4227],[180,24.2598],[181,24.4556],[182,21.0778],[183,25.5202],[184,23.2654]]
       }
       thumb_03_l_ctrl.Scale.X {type=MovieSceneScriptingFloatChannel, default=1, const [1,185]}
       thumb_03_l_ctrl.Scale.Y {type=MovieSceneScriptingFloatChannel, default=1, const [1,185]}
       thumb_03_l_ctrl.Scale.Z {type=MovieSceneScriptingFloatChannel, default=1, const [1,185]}
       index_metacarpal_l_ctrl.Location.X {
        type=MovieSceneScriptingFloatChannel
        default=-0.0001
        thinned_from=185
        keys [[0,-0.0001],[56,-0.0004],[61,-0.0011],[68,-0.0007],[70,-0.0002],[72,-0.0006],[74,-0.0001],[75,-0.0002],[76,-0.0002],[77,-0.0006],[78,-0.0002],[80,-0.0002],[82,-0.0004],[83,-0.0001],[165,-0.0001],[166,-0.0013],[167,-0.0001],[168,-0.0012],[169,-0.0006],[172,-0.0001],[180,-0.0002],[181,-0.0008],[182,-0.0008],[183,0.0001],[184,-0.0001]]
       }
       index_metacarpal_l_ctrl.Location.Y {
        type=MovieSceneScriptingFloatChannel
        default=0.0001
        thinned_from=185
        keys [[0,0.0001],[57,-0.0003],[68,-0.0004],[70,0],[72,-0.0005],[74,0.0001],[75,-0.0001],[76,0],[77,-0.0004],[78,0],[82,-0.0002],[111,0],[165,0.0001],[166,0.0006],[167,0.0001],[168,0.001],[171,0.0005],[172,0.0001],[181,0.0001],[182,-0.0001],[184,0.0001]]
       }
       index_metacarpal_l_ctrl.Location.Z {
        type=MovieSceneScriptingFloatChannel
        default=0
        thinned_from=185
        keys [[0,0],[49,0],[50,0.0002],[55,0.0005],[61,0.0024],[64,0.0022],[68,0.0011],[70,0.0002],[72,0.0011],[74,0],[75,0.0001],[76,0],[77,0.0004],[78,0.0001],[82,0.0002],[165,0],[166,0.0018],[167,0],[168,0.0011],[171,0.0004],[172,0.0001],[180,0.0002],[181,0.0012],[182,0.0017],[183,0.0002],[184,0]]
       }
       index_metacarpal_l_ctrl.Rotation.X {type=MovieSceneScriptingFloatChannel, default=0.0061, const [0.0061,185]}
       index_metacarpal_l_ctrl.Rotation.Y {type=MovieSceneScriptingFloatChannel, default=-0.47, const [-0.47,185]}
       index_metacarpal_l_ctrl.Rotation.Z {type=MovieSceneScriptingFloatChannel, default=-0.7433, const [-0.7433,185]}
       index_metacarpal_l_ctrl.Scale.X {type=MovieSceneScriptingFloatChannel, default=1, const [1,185]}
       index_metacarpal_l_ctrl.Scale.Y {type=MovieSceneScriptingFloatChannel, default=1, const [1,185]}
       index_metacarpal_l_ctrl.Scale.Z {type=MovieSceneScriptingFloatChannel, default=1, const [1,185]}
       index_01_l_ctrl.Location.X {
        type=MovieSceneScriptingFloatChannel
        default=0
        thinned_from=185
        keys [[0,0],[48,-0.0002],[55,-0.0014],[56,-0.0017],[60,-0.0045],[61,-0.005],[62,-0.0052],[64,-0.0048],[68,-0.0031],[69,-0.0017],[70,-0.0005],[72,-0.003],[74,0],[75,-0.0009],[76,-0.0004],[77,-0.0026],[78,-0.0006],[82,-0.0015],[83,-0.0003],[84,-0.0004],[85,0],[165,0],[166,-0.0055],[167,0],[168,-0.0052],[169,-0.0034],[171,-0.002],[172,-0.0002],[176,0],[178,-0.0001],[179,0],[180,-0.0006],[181,-0.0034],[182,-0.0016],[183,-0.0005],[184,0]]
       }
       index_01_l_ctrl.Location.Y {
        type=MovieSceneScriptingFloatChannel
        default=0
        thinned_from=185
        keys [[0,0],[50,-0.0004],[55,-0.001],[57,-0.0017],[60,-0.0033],[61,-0.0037],[62,-0.0038],[64,-0.0035],[68,-0.0017],[69,-0.0009],[70,-0.0003],[72,-0.0018],[74,0],[75,-0.0003],[76,-0.0001],[77,-0.0006],[78,-0.0001],[84,-0.0003],[85,0],[165,0],[166,-0.0011],[167,0],[168,-0.0008],[169,-0.0001],[170,0.0003],[171,0.0004],[172,0.0001],[180,-0.0001],[182,-0.0025],[183,-0.0001],[184,0]]
       }
       index_01_l_ctrl.Location.Z {
        type=MovieSceneScriptingFloatChannel
        default=0
        thinned_from=185
        keys [[0,0],[65,0.0005],[82,0.0001],[83,-0.0003],[84,-0.0003],[85,0],[165,0],[166,0.0004],[168,-0.0002],[169,0.0001],[171,-0.0003],[172,0],[180,0],[182,0.0011],[183,-0.0007],[184,0]]
       }
       index_01_l_ctrl.Rotation.X {
        type=MovieSceneScriptingFloatChannel
        default=-2.076
        thinned_from=185
        keys [[0,1.0239],[1,1.0523],[2,1.0144],[3,1.0499],[4,1.0172],[5,1.0149],[6,1.047],[7,1.0555],[8,1.0427],[9,1.0545],[10,1.0202],[11,1.0305],[12,1.0239],[13,1.0116],[14,1.0467],[15,1.0601],[16,1.0643],[17,1.0793],[18,1.0537],[19,1.0185],[20,1.0321],[21,1.0279],[22,1.0134],[23,1.0087],[24,0.9872],[25,1.0133],[26,1.0094],[27,1.0111],[28,0.9843],[29,0.9943],[30,0.9941],[31,0.9468],[32,0.9981],[33,0.9982],[34,0.9806],[35,0.994],[36,1.0233],[37,0.9987],[38,0.9881],[39,1.0194],[40,1.0036],[41,1.0082],[42,1.024],[43,1.0296],[44,1.038],[45,1.0123],[46,0.9383],[47,0.8208],[48,0.6647],[49,0.4749],[50,0.2561],[51,0.0131],[52,-0.2492],[53,-0.5259],[54,-0.8123],[55,-1.1035],[56,-1.6527],[57,-2.5989],[58,-3.7668],[59,-4.9807],[60,-6.0651],[61,-6.8445],[62,-7.1433],[63,-7.1121],[64,-7.0301],[65,-6.9145],[67,-6.6514],[68,-6.5383],[70,-6.3569],[73,-6.1611],[74,-6.0833],[75,-5.9703],[76,-5.8244],[77,-5.6408],[78,-5.274],[82,-5.1687],[83,-4.7355],[84,-4.928],[85,-4.7325],[86,-4.4882],[87,-4.4166],[88,-4.3541],[89,-4.3155],[90,-4.2866],[91,-4.2677],[96,-4.251],[97,-4.2777],[98,-4.2915],[99,-4.2421],[100,-4.0346],[101,-4.0834],[103,-4.2179],[104,-4.3238],[105,-4.3373],[106,-4.3286],[107,-4.2721],[108,-4.1993],[109,-4.2202],[110,-4.2148],[111,-4.1512],[112,-3.8596],[113,-4.0861],[114,-4.3235],[115,-4.5959],[116,-4.6024],[117,-4.5747],[118,-4.5687],[119,-4.6172],[120,-4.6031],[121,-4.643],[122,-4.66],[123,-4.657],[124,-4.6642],[125,-4.6256],[126,-4.6512],[128,-4.668],[129,-4.6941],[132,-4.6543],[133,-4.669],[135,-4.7201],[136,-4.6913],[137,-4.672],[138,-4.6947],[139,-4.658],[140,-4.6744],[141,-4.6701],[142,-4.6377],[143,-4.6791],[144,-4.7385],[145,-4.7844],[146,-4.7539],[147,-4.7719],[148,-4.7178],[149,-4.7554],[150,-4.7651],[151,-4.7378],[152,-4.7256],[153,-4.7383],[154,-4.7721],[155,-4.7799],[156,-4.7579],[158,-4.7821],[159,-4.7865],[160,-4.8028],[161,-4.7421],[164,-4.7886],[165,-4.8353],[166,-5.1161],[167,-5.4496],[168,-5.2814],[169,-4.4591],[170,-3.1406],[171,-1.5683],[172,-0.7147],[173,-0.0813],[174,0.0995],[175,0.1662],[176,0.0811],[177,0.0273],[178,0.0588],[179,0.0651],[180,-0.0419],[181,-0.1721],[183,-2.0163],[184,-2.076]]
       }
       index_01_l_ctrl.Rotation.Y {
        type=MovieSceneScriptingFloatChannel
        default=-4.0677
        thinned_from=185
        keys [[0,-4.6847],[2,-4.658],[4,-4.6892],[6,-4.6196],[7,-4.6528],[9,-4.6732],[10,-4.6509],[11,-4.6125],[12,-4.6617],[13,-4.6511],[14,-4.6595],[15,-4.7194],[16,-4.6774],[18,-4.6384],[19,-4.6522],[20,-4.6162],[22,-4.6624],[23,-4.6736],[24,-4.624],[25,-4.6755],[26,-4.6939],[27,-4.7323],[28,-4.6738],[29,-4.7078],[30,-4.6927],[31,-4.6193],[32,-4.6789],[33,-4.6534],[34,-4.6668],[35,-4.6349],[36,-4.6386],[37,-4.6685],[38,-4.6388],[39,-4.5982],[40,-4.6794],[41,-4.6267],[42,-4.6709],[43,-4.6486],[44,-4.6589],[46,-4.6453],[48,-4.6081],[51,-4.5194],[55,-4.3675],[56,-4.2928],[57,-4.164],[59,-3.8399],[60,-3.6924],[61,-3.5863],[62,-3.5457],[63,-3.5584],[64,-3.591],[67,-3.729],[68,-3.7617],[69,-3.7836],[70,-3.7936],[72,-3.6967],[74,-3.6875],[75,-3.7196],[76,-3.6643],[77,-3.5789],[78,-3.7308],[82,-3.7295],[83,-4.0526],[84,-3.8795],[85,-3.6739],[86,-3.6475],[87,-3.5965],[96,-3.6134],[97,-3.6002],[98,-3.756],[99,-3.9342],[100,-4.0862],[101,-4.1417],[102,-4.1888],[103,-4.1689],[104,-4.0749],[105,-4.0312],[106,-4.0478],[107,-4.1039],[108,-4.1333],[109,-4.1763],[110,-4.1269],[111,-4.1561],[112,-4.4437],[113,-4.2452],[114,-4.0874],[115,-3.8551],[116,-3.8901],[117,-3.8488],[118,-3.827],[120,-3.8494],[121,-3.8394],[122,-3.8393],[123,-3.8943],[124,-3.9022],[125,-3.8684],[126,-3.9659],[127,-3.9202],[128,-3.9269],[129,-3.847],[130,-3.872],[131,-3.8546],[132,-3.8147],[133,-3.8414],[134,-3.8498],[135,-3.875],[136,-3.8267],[137,-3.8626],[138,-3.8042],[140,-3.8881],[141,-3.8207],[142,-3.7971],[143,-3.8383],[144,-3.8098],[145,-3.9126],[146,-3.8187],[147,-3.841],[148,-3.8066],[149,-3.8016],[150,-3.7631],[151,-3.75],[153,-3.7583],[154,-3.7917],[155,-3.782],[156,-3.7141],[157,-3.7164],[158,-3.7266],[160,-3.7306],[161,-3.7559],[162,-3.7637],[163,-3.7875],[164,-3.7538],[165,-3.756],[166,-3.624],[167,-3.6347],[168,-4.0151],[169,-4.6155],[170,-4.6937],[171,-4.5795],[172,-4.6318],[173,-4.7467],[174,-4.8299],[175,-4.8054],[176,-4.7986],[177,-4.7249],[178,-4.728],[179,-4.6526],[180,-4.6015],[181,-4.4357],[182,-4.2246],[183,-4.1363],[184,-4.0677]]
       }
       index_01_l_ctrl.Rotation.Z {
        type=MovieSceneScriptingFloatChannel
        default=-46.4651
        thinned_from=185
        keys [[0,-58.7333],[1,-58.8639],[2,-58.7459],[3,-58.9136],[4,-58.7207],[5,-58.7217],[6,-58.8916],[9,-58.8793],[10,-58.7784],[13,-58.7483],[15,-59.0058],[18,-58.9298],[19,-58.7339],[21,-58.8036],[24,-58.5667],[25,-58.6733],[30,-58.5638],[31,-58.414],[32,-58.6138],[33,-58.6412],[34,-58.4792],[36,-58.772],[38,-58.6454],[39,-58.769],[41,-58.6768],[44,-58.8039],[45,-58.6941],[46,-58.3786],[47,-57.878],[48,-57.2127],[49,-56.4034],[51,-54.4349],[54,-50.9164],[55,-49.675],[56,-47.3343],[57,-43.3006],[58,-38.3223],[59,-33.1478],[60,-28.5252],[61,-25.2029],[62,-23.9292],[63,-24.085],[64,-24.4952],[65,-25.0744],[67,-26.3968],[68,-26.9688],[70,-27.9249],[72,-28.6591],[74,-29.5839],[75,-30.2395],[76,-31.0213],[77,-32.2431],[78,-33.8265],[79,-34.0482],[81,-34.3515],[82,-34.5732],[83,-36.9075],[84,-35.8272],[85,-36.3431],[86,-37.3241],[88,-37.7835],[91,-38.0582],[96,-38.1074],[97,-37.8054],[98,-37.9731],[99,-38.4254],[100,-39.2319],[102,-38.8959],[104,-38.3671],[105,-38.2762],[106,-38.3021],[108,-38.654],[109,-38.6639],[110,-38.5926],[111,-38.7451],[112,-39.7076],[114,-38.2431],[115,-37.3401],[116,-37.3101],[117,-37.4146],[119,-37.2817],[120,-37.2781],[122,-37.0565],[123,-37.2015],[124,-37.1717],[125,-37.2934],[127,-37.1365],[128,-37.1338],[129,-36.9754],[130,-36.9142],[135,-36.8088],[136,-36.9123],[137,-36.9279],[138,-36.825],[139,-36.9246],[143,-36.973],[144,-36.8598],[145,-37.0005],[146,-36.8247],[147,-36.8786],[148,-36.8567],[149,-36.7254],[150,-36.6804],[153,-36.7266],[159,-36.4855],[160,-36.3729],[161,-36.5524],[163,-36.4702],[164,-36.3523],[165,-36.1447],[166,-34.8819],[167,-33.3623],[168,-33.7997],[169,-37.1053],[170,-42.3651],[171,-48.7056],[172,-52.0191],[173,-54.5416],[174,-55.5483],[175,-55.9383],[177,-55.3518],[179,-55.2458],[180,-54.8975],[181,-54.3567],[182,-50.2107],[183,-47.1449],[184,-46.4651]]
       }
       index_01_l_ctrl.Scale.X {type=MovieSceneScriptingFloatChannel, default=1, const [1,185]}
       index_01_l_ctrl.Scale.Y {type=MovieSceneScriptingFloatChannel, default=1, const [1,185]}
       index_01_l_ctrl.Scale.Z {type=MovieSceneScriptingFloatChannel, default=1, const [1,185]}
       index_02_l_ctrl.Location.X {
        type=MovieSceneScriptingFloatChannel
        default=0
        thinned_from=185
        keys [[0,0],[76,0],[77,-0.0003],[84,-0.0003],[85,0],[165,0],[166,-0.0005],[167,0],[168,-0.0007],[171,-0.0003],[172,0],[181,0.0001],[182,-0.0006],[184,0]]
       }
       index_02_l_ctrl.Location.Y {
        type=MovieSceneScriptingFloatChannel
        default=0
        thinned_from=185
        keys [[0,0],[56,-0.0004],[62,-0.0012],[68,-0.0007],[70,-0.0001],[72,-0.0006],[74,0],[75,-0.0002],[76,-0.0001],[77,-0.0005],[78,-0.0001],[82,-0.0003],[165,0],[166,-0.0016],[167,0],[168,-0.0012],[169,-0.0007],[171,-0.0005],[172,-0.0001],[180,-0.0001],[181,-0.0002],[182,0.0004],[183,0],[184,0]]
       }
       index_02_l_ctrl.Location.Z {
        type=MovieSceneScriptingFloatChannel
        default=0
        thinned_from=185
        keys [[0,0],[76,0],[77,-0.0003],[78,-0.0001],[84,-0.0002],[85,0],[165,0],[166,0.0008],[167,0],[168,0.0006],[170,0.0001],[180,0.0001],[181,0.0002],[182,-0.0001],[184,0]]
       }
       index_02_l_ctrl.Rotation.X {
        type=MovieSceneScriptingFloatChannel
        default=-6.5969
        thinned_from=185
        keys [[0,-8.4994],[1,-8.5066],[3,-8.4772],[5,-8.5052],[8,-8.4791],[9,-8.503],[13,-8.4952],[14,-8.5129],[15,-8.4731],[17,-8.4647],[18,-8.4875],[20,-8.4992],[23,-8.4716],[24,-8.4889],[25,-8.4694],[26,-8.4803],[29,-8.4364],[31,-8.4563],[32,-8.4376],[34,-8.4626],[36,-8.4277],[38,-8.4505],[40,-8.4255],[41,-8.4406],[42,-8.4219],[43,-8.4236],[44,-8.45],[45,-8.4335],[46,-8.386],[47,-8.3107],[48,-8.2106],[49,-8.0888],[51,-7.7926],[54,-7.2632],[55,-7.0764],[56,-6.7242],[57,-6.1173],[58,-5.3682],[59,-4.5897],[60,-3.8941],[61,-3.3942],[62,-3.2026],[63,-3.2156],[64,-3.2497],[67,-3.4044],[68,-3.4488],[70,-3.5035],[72,-3.5788],[73,-3.6265],[74,-3.6551],[75,-3.6492],[76,-3.6977],[77,-3.8394],[78,-3.9276],[82,-4.0146],[83,-4.0578],[84,-4.3121],[85,-4.4987],[86,-4.5473],[88,-4.6008],[90,-4.6232],[97,-4.6297],[98,-4.5927],[100,-4.4901],[101,-4.4591],[102,-4.4052],[104,-4.4178],[105,-4.4436],[107,-4.4392],[108,-4.4266],[109,-4.4345],[110,-4.4171],[111,-4.4203],[112,-4.4994],[113,-4.4749],[114,-4.4028],[115,-4.3098],[116,-4.3205],[117,-4.3014],[118,-4.3166],[121,-4.294],[122,-4.2947],[126,-4.2416],[127,-4.2422],[128,-4.2323],[129,-4.2623],[130,-4.2516],[132,-4.2688],[133,-4.2698],[135,-4.2391],[137,-4.2432],[140,-4.2166],[141,-4.2455],[142,-4.2578],[143,-4.2157],[144,-4.198],[145,-4.1633],[146,-4.1912],[147,-4.1824],[148,-4.2092],[149,-4.1985],[150,-4.1978],[151,-4.2212],[152,-4.227],[153,-4.2012],[155,-4.1787],[156,-4.2151],[158,-4.2135],[160,-4.1924],[161,-4.2069],[164,-4.2013],[165,-4.164],[166,-4.0797],[167,-3.9051],[168,-3.9531],[169,-4.3347],[170,-5.0687],[171,-6.0234],[172,-6.5153],[173,-6.871],[175,-7.1211],[177,-7.1604],[178,-7.1685],[179,-7.1974],[180,-7.1594],[181,-7.1809],[182,-7.0444],[183,-6.4372],[184,-6.5969]]
       }
       index_02_l_ctrl.Rotation.Y {
        type=MovieSceneScriptingFloatChannel
        default=6.1716
        thinned_from=185
        keys [[0,6.2535],[44,6.257],[46,6.2355],[49,6.1515],[55,5.8708],[56,5.7721],[57,5.6024],[59,5.1756],[60,4.9813],[61,4.8417],[62,4.7882],[63,4.7969],[65,4.8517],[68,4.9524],[72,5.0362],[73,5.0649],[74,5.0817],[75,5.0786],[76,5.108],[77,5.1935],[78,5.2418],[82,5.2903],[83,5.3125],[84,5.441],[85,5.5313],[88,5.578],[97,5.5911],[102,5.4884],[104,5.4944],[105,5.5066],[109,5.5023],[111,5.4955],[112,5.5324],[113,5.5212],[114,5.4873],[115,5.4422],[116,5.4475],[117,5.4381],[128,5.4036],[129,5.4187],[133,5.4224],[135,5.4071],[140,5.3956],[142,5.4164],[143,5.3952],[145,5.3683],[146,5.3827],[147,5.3782],[148,5.3919],[150,5.3861],[152,5.4009],[155,5.3763],[156,5.3949],[160,5.3833],[164,5.3879],[165,5.3687],[166,5.3304],[167,5.2287],[168,5.2535],[169,5.4508],[170,5.7606],[171,6.0522],[172,6.1572],[173,6.2104],[175,6.2389],[181,6.244],[182,6.2315],[183,6.1442],[184,6.1716]]
       }
       index_02_l_ctrl.Rotation.Z {
        type=MovieSceneScriptingFloatChannel
        default=-27.4416
        thinned_from=185
        keys [[0,-40.9618],[1,-41.0129],[3,-40.8037],[4,-40.946],[5,-41.0029],[8,-40.8177],[9,-40.9875],[13,-40.9317],[14,-41.0578],[15,-40.7751],[17,-40.7152],[18,-40.8772],[20,-40.9601],[21,-40.8437],[23,-40.764],[24,-40.8872],[25,-40.7484],[26,-40.8258],[28,-40.5701],[29,-40.5142],[31,-40.6557],[32,-40.5224],[33,-40.5617],[34,-40.7003],[36,-40.4524],[38,-40.6141],[40,-40.4369],[41,-40.5436],[42,-40.4115],[43,-40.4236],[44,-40.6107],[45,-40.4868],[46,-40.1307],[47,-39.5656],[48,-38.8147],[49,-37.9013],[50,-36.8485],[51,-35.6796],[52,-34.4178],[55,-30.3073],[56,-27.6654],[57,-23.1128],[58,-17.4941],[59,-11.6537],[60,-6.4364],[61,-2.6867],[62,-1.2491],[63,-1.361],[64,-1.6541],[65,-2.0647],[66,-2.5289],[67,-2.9831],[68,-3.3634],[69,-3.607],[70,-3.8258],[71,-4.1284],[72,-4.453],[73,-4.8516],[74,-5.092],[75,-5.0425],[76,-5.4445],[77,-6.608],[78,-7.3261],[79,-7.5453],[81,-7.8517],[82,-8.0295],[83,-8.3685],[84,-10.3848],[85,-11.8595],[86,-12.2377],[87,-12.4744],[88,-12.6537],[89,-12.7731],[91,-12.8589],[97,-12.8789],[98,-12.5908],[99,-12.1688],[100,-11.7943],[101,-11.5522],[102,-11.1315],[103,-11.2009],[104,-11.2298],[105,-11.4319],[106,-11.3907],[107,-11.3975],[108,-11.2992],[109,-11.3606],[110,-11.2247],[111,-11.2497],[112,-11.8673],[113,-11.6763],[114,-11.1122],[115,-10.3811],[116,-10.4658],[117,-10.3147],[118,-10.4346],[120,-10.342],[121,-10.2563],[122,-10.2624],[124,-10.0883],[126,-9.8418],[127,-9.8466],[128,-9.7684],[129,-10.0058],[130,-9.9216],[131,-9.9649],[132,-10.0572],[133,-10.0657],[135,-9.8226],[137,-9.8544],[139,-9.6841],[140,-9.6439],[141,-9.8728],[142,-9.9705],[143,-9.6365],[144,-9.4961],[145,-9.2197],[146,-9.4424],[147,-9.3724],[148,-9.5849],[149,-9.5001],[150,-9.4948],[151,-9.68],[152,-9.7262],[153,-9.5213],[154,-9.4002],[155,-9.343],[156,-9.6319],[157,-9.5947],[158,-9.6193],[159,-9.5624],[160,-9.4515],[161,-9.5667],[162,-9.5556],[163,-9.5664],[164,-9.5222],[165,-9.2254],[166,-8.55],[167,-7.1433],[168,-7.5312],[169,-10.5744],[170,-16.219],[171,-23.2934],[172,-26.8556],[173,-29.4036],[174,-30.3414],[175,-31.186],[176,-31.3576],[178,-31.523],[179,-31.7283],[180,-31.4587],[181,-31.6155],[182,-30.6645],[183,-26.2949],[184,-27.4416]]
       }
       index_02_l_ctrl.Scale.X {type=MovieSceneScriptingFloatChannel, default=1, const [1,185]}
       index_02_l_ctrl.Scale.Y {type=MovieSceneScriptingFloatChannel, default=1, const [1,185]}
       index_02_l_ctrl.Scale.Z {type=MovieSceneScriptingFloatChannel, default=1, const [1,185]}
       index_03_l_ctrl.Location.X {
        type=MovieSceneScriptingFloatChannel
        default=-0.0001
        thinned_from=185
        keys [[0,-0.0001],[72,-0.0003],[76,-0.0001],[84,-0.0003],[85,-0.0001],[165,-0.0001],[166,-0.0008],[167,-0.0001],[168,-0.0004],[170,-0.0001],[180,-0.0002],[181,-0.001],[182,-0.0012],[183,-0.0004],[184,-0.0001]]
       }
       index_03_l_ctrl.Location.Y {
        type=MovieSceneScriptingFloatChannel
        default=0.0002
        thinned_from=185
        keys [[0,0.0003],[72,0.0005],[167,0.0003],[168,0.0001],[181,0.0003],[182,-0.0004],[184,0.0002]]
       }
       index_03_l_ctrl.Location.Z {
        type=MovieSceneScriptingFloatChannel
        default=0
        thinned_from=185
        keys [[0,0],[76,-0.0001],[77,-0.0002],[165,0],[166,0.001],[167,0],[168,0.0006],[170,-0.0002],[183,0],[184,0]]
       }
       index_03_l_ctrl.Rotation.X {
        type=MovieSceneScriptingFloatChannel
        default=-1.4251
        thinned_from=185
        keys [[0,-1.9928],[1,-1.9942],[2,-1.9777],[3,-1.9883],[5,-1.9836],[6,-1.9935],[8,-1.9691],[9,-2.0025],[10,-1.9928],[11,-1.9692],[12,-1.9695],[13,-1.9516],[14,-1.9756],[15,-1.9683],[16,-1.9545],[18,-1.9805],[19,-1.9806],[20,-1.9644],[23,-1.9638],[24,-1.9752],[25,-1.9657],[26,-1.9888],[28,-1.9633],[29,-1.9614],[30,-1.9659],[31,-1.957],[32,-1.9582],[33,-1.9431],[34,-1.9641],[35,-1.9692],[36,-1.9648],[37,-1.9522],[38,-1.9607],[39,-1.9583],[40,-1.9429],[41,-1.9682],[42,-1.9635],[43,-1.9647],[44,-1.9614],[45,-1.9526],[46,-1.9369],[47,-1.9148],[48,-1.8869],[49,-1.8539],[50,-1.8163],[52,-1.7298],[54,-1.6324],[55,-1.5811],[56,-1.4843],[57,-1.3181],[58,-1.1131],[59,-0.9003],[60,-0.7102],[61,-0.5736],[62,-0.5212],[63,-0.5228],[64,-0.527],[67,-0.5462],[68,-0.5519],[70,-0.5604],[72,-0.5668],[73,-0.574],[74,-0.5825],[75,-0.5883],[76,-0.5892],[77,-0.6007],[78,-0.6027],[79,-0.607],[83,-0.6325],[84,-0.7774],[85,-0.8319],[86,-0.8785],[88,-0.9059],[89,-0.914],[90,-0.9175],[96,-0.9199],[97,-0.9009],[98,-0.8868],[99,-0.8658],[100,-0.8543],[101,-0.8507],[102,-0.8353],[103,-0.8415],[105,-0.8418],[106,-0.8499],[108,-0.8568],[109,-0.8582],[110,-0.8562],[111,-0.8638],[112,-0.9681],[113,-0.9125],[114,-0.8399],[115,-0.7206],[116,-0.7215],[118,-0.7261],[119,-0.7254],[121,-0.7171],[124,-0.709],[125,-0.7107],[126,-0.7094],[127,-0.7104],[129,-0.7164],[130,-0.714],[131,-0.7141],[132,-0.7197],[133,-0.7183],[134,-0.7073],[136,-0.7108],[137,-0.7204],[138,-0.7093],[139,-0.7171],[140,-0.7075],[141,-0.7196],[142,-0.7237],[143,-0.7057],[144,-0.6975],[145,-0.684],[146,-0.7053],[147,-0.6906],[148,-0.7116],[149,-0.7076],[150,-0.7075],[152,-0.712],[153,-0.7069],[155,-0.7105],[156,-0.721],[158,-0.7206],[159,-0.7168],[160,-0.7212],[161,-0.7374],[162,-0.7503],[164,-0.7582],[165,-0.7563],[166,-0.7084],[167,-0.6555],[168,-0.6597],[169,-0.7598],[170,-0.9668],[171,-1.2657],[172,-1.4541],[173,-1.5488],[174,-1.5771],[175,-1.6183],[176,-1.6026],[177,-1.6117],[178,-1.6285],[179,-1.6523],[180,-1.6315],[181,-1.622],[182,-1.5974],[183,-1.3576],[184,-1.4251]]
       }
       index_03_l_ctrl.Rotation.Y {
        type=MovieSceneScriptingFloatChannel
        default=2.7655
        thinned_from=185
        keys [[0,3.5004],[1,3.5019],[2,3.4829],[3,3.4952],[5,3.4898],[6,3.5011],[8,3.473],[9,3.5115],[10,3.5002],[11,3.473],[12,3.4735],[13,3.4526],[14,3.4806],[15,3.4721],[16,3.456],[18,3.4862],[19,3.4862],[20,3.4675],[23,3.4668],[24,3.48],[25,3.4691],[26,3.4957],[28,3.4662],[29,3.4641],[30,3.4693],[31,3.4589],[32,3.4604],[33,3.4427],[34,3.4671],[35,3.4731],[36,3.4681],[37,3.4534],[38,3.4632],[39,3.4605],[40,3.4424],[41,3.472],[42,3.4665],[44,3.464],[45,3.4515],[46,3.426],[47,3.389],[48,3.3415],[49,3.2847],[50,3.2198],[52,3.0703],[55,2.8147],[56,2.649],[57,2.3642],[58,2.0129],[59,1.6479],[60,1.3219],[61,1.0877],[62,0.9979],[63,1.0022],[64,1.0136],[65,1.0297],[67,1.066],[68,1.0815],[70,1.1042],[72,1.1206],[73,1.1397],[74,1.1619],[75,1.1768],[76,1.1792],[77,1.2088],[78,1.2138],[79,1.2247],[82,1.2731],[83,1.2844],[84,1.6145],[85,1.7343],[86,1.8318],[88,1.8872],[89,1.9035],[90,1.9105],[96,1.9151],[97,1.8779],[98,1.8495],[99,1.8072],[100,1.7842],[101,1.7768],[102,1.7451],[103,1.758],[105,1.7585],[106,1.7751],[108,1.7893],[109,1.7922],[110,1.7881],[111,1.8035],[112,2.0078],[113,1.9007],[114,1.7547],[115,1.4962],[116,1.4984],[118,1.5088],[119,1.5072],[121,1.4884],[124,1.4698],[125,1.4737],[126,1.4706],[127,1.4729],[129,1.4866],[130,1.4811],[131,1.4815],[132,1.4943],[133,1.4911],[134,1.4659],[136,1.4741],[137,1.4959],[138,1.4705],[139,1.4884],[140,1.4663],[141,1.4939],[142,1.5033],[143,1.4621],[144,1.4432],[145,1.4116],[146,1.4612],[147,1.4272],[148,1.4758],[149,1.4665],[150,1.4664],[152,1.4766],[153,1.4649],[155,1.4733],[156,1.4971],[158,1.4963],[159,1.4875],[160,1.4977],[161,1.5341],[162,1.5628],[164,1.5803],[165,1.5761],[166,1.4693],[167,1.3441],[168,1.3529],[169,1.5814],[170,2.0016],[171,2.5204],[172,2.8077],[173,2.9416],[174,2.9806],[175,3.0362],[176,3.0152],[177,3.0274],[178,3.0499],[179,3.0815],[180,3.0541],[181,3.042],[182,3.0175],[183,2.6642],[184,2.7655]]
       }
       index_03_l_ctrl.Rotation.Z {
        type=MovieSceneScriptingFloatChannel
        default=3.0783
        thinned_from=185
        keys [[0,-3.665],[1,-3.6804],[2,-3.4976],[3,-3.6153],[4,-3.5993],[5,-3.5631],[6,-3.6722],[7,-3.5546],[8,-3.4017],[9,-3.772],[10,-3.6641],[11,-3.4027],[12,-3.4067],[13,-3.207],[14,-3.4744],[15,-3.393],[16,-3.2397],[17,-3.4036],[18,-3.529],[19,-3.5291],[20,-3.3493],[21,-3.3709],[23,-3.3428],[24,-3.4695],[25,-3.3643],[26,-3.6202],[27,-3.4647],[28,-3.3371],[29,-3.3166],[30,-3.3665],[31,-3.2671],[32,-3.2809],[33,-3.1117],[34,-3.3461],[35,-3.4031],[36,-3.3546],[37,-3.2137],[38,-3.3085],[39,-3.2821],[40,-3.1099],[41,-3.3922],[42,-3.3395],[43,-3.3533],[44,-3.316],[45,-3.2039],[46,-2.9857],[47,-2.6709],[48,-2.2692],[49,-1.7901],[50,-1.2433],[51,-0.6384],[52,0.0151],[53,0.7075],[54,1.4293],[55,2.1708],[56,3.57],[57,5.9744],[58,8.9392],[59,12.0195],[60,14.7706],[61,16.7474],[62,17.5052],[63,17.4715],[64,17.3828],[68,16.8553],[70,16.6763],[72,16.5562],[74,16.2283],[75,16.1132],[76,16.0947],[77,15.868],[78,15.8274],[79,15.7421],[82,15.3681],[83,15.2756],[84,12.6659],[85,11.7078],[86,10.9253],[88,10.4779],[89,10.3459],[90,10.2889],[96,10.2514],[97,10.5543],[98,10.7842],[99,11.1256],[100,11.3118],[101,11.3719],[102,11.6269],[103,11.5234],[105,11.5192],[106,11.3851],[107,11.3165],[109,11.2474],[110,11.2808],[111,11.1551],[112,9.4929],[113,10.369],[114,11.5498],[115,13.6112],[116,13.5939],[118,13.5111],[119,13.5241],[121,13.6726],[124,13.8196],[125,13.7891],[126,13.813],[127,13.795],[129,13.6865],[130,13.7301],[131,13.7266],[132,13.6261],[133,13.6512],[134,13.8501],[136,13.7856],[137,13.6132],[138,13.8137],[139,13.6724],[140,13.8474],[141,13.6294],[142,13.5546],[143,13.8802],[144,14.0293],[145,14.2778],[146,13.887],[147,14.1553],[148,13.7724],[149,13.8454],[150,13.8463],[152,13.7659],[153,13.8577],[155,13.7918],[156,13.6036],[158,13.6103],[159,13.6795],[160,13.5995],[161,13.311],[162,13.083],[164,12.9437],[165,12.9771],[166,13.8545],[167,14.8081],[168,14.7247],[169,12.9078],[170,9.5098],[171,5.1748],[172,2.7038],[173,1.5152],[174,1.1667],[175,0.6652],[176,0.8558],[177,0.7453],[178,0.5418],[179,0.2546],[180,0.5025],[181,0.6121],[182,0.859],[183,3.9598],[184,3.0783]]
       }
       index_03_l_ctrl.Scale.X {type=MovieSceneScriptingFloatChannel, default=1, const [1,185]}
       index_03_l_ctrl.Scale.Y {type=MovieSceneScriptingFloatChannel, default=1, const [1,185]}
       index_03_l_ctrl.Scale.Z {type=MovieSceneScriptingFloatChannel, default=1, const [1,185]}
       middle_metacarpal_l_ctrl.Location.X {
        type=MovieSceneScriptingFloatChannel
        default=-0.0002
        thinned_from=185
        keys [[0,-0.0002],[50,-0.0003],[57,-0.0008],[61,-0.0016],[64,-0.0016],[68,-0.001],[70,-0.0003],[72,-0.001],[74,-0.0002],[75,-0.0004],[76,-0.0003],[77,-0.0009],[78,-0.0003],[82,-0.0006],[83,-0.0002],[165,-0.0002],[166,-0.0017],[167,-0.0002],[168,-0.0016],[169,-0.001],[171,-0.0006],[172,-0.0002],[180,-0.0003],[181,-0.0011],[183,-0.0003],[184,-0.0002]]
       }
       middle_metacarpal_l_ctrl.Location.Y {
        type=MovieSceneScriptingFloatChannel
        default=0
        thinned_from=185
        keys [[0,0],[64,-0.0007],[70,-0.0001],[72,-0.0003],[74,0],[165,-0.0001],[166,0.0001],[167,0],[168,0.0003],[171,0.0003],[172,0],[180,0],[181,-0.0001],[182,-0.0004],[183,0],[184,0]]
       }
       middle_metacarpal_l_ctrl.Location.Z {
        type=MovieSceneScriptingFloatChannel
        default=0
        thinned_from=185
        keys [[0,0],[64,0.0005],[70,0],[72,0.0001],[165,-0.0001],[166,0.0004],[167,0],[168,0.0001],[180,0],[182,0.0005],[183,-0.0002],[184,0]]
       }
       middle_metacarpal_l_ctrl.Rotation.X {type=MovieSceneScriptingFloatChannel, default=-0.0682, const [-0.0682,185]}
       middle_metacarpal_l_ctrl.Rotation.Y {type=MovieSceneScriptingFloatChannel, default=1.7387, const [1.7387,185]}
       middle_metacarpal_l_ctrl.Rotation.Z {type=MovieSceneScriptingFloatChannel, default=-2.245, const [-2.245,185]}
       middle_metacarpal_l_ctrl.Scale.X {type=MovieSceneScriptingFloatChannel, default=1, const [1,185]}
       middle_metacarpal_l_ctrl.Scale.Y {type=MovieSceneScriptingFloatChannel, default=1, const [1,185]}
       middle_metacarpal_l_ctrl.Scale.Z {type=MovieSceneScriptingFloatChannel, default=1, const [1,185]}
       middle_01_l_ctrl.Location.X {
        type=MovieSceneScriptingFloatChannel
        default=0
        thinned_from=185
        keys [[0,0],[48,-0.0002],[55,-0.0012],[56,-0.0015],[60,-0.004],[61,-0.0045],[62,-0.0046],[64,-0.0044],[68,-0.0028],[69,-0.0015],[70,-0.0005],[72,-0.0027],[74,0],[75,-0.0008],[76,-0.0004],[77,-0.0023],[78,-0.0005],[82,-0.0013],[83,-0.0003],[84,-0.0004],[85,0],[165,0],[166,-0.0049],[167,0],[168,-0.0046],[169,-0.0031],[171,-0.0018],[172,-0.0002],[176,0],[179,0],[180,-0.0005],[181,-0.003],[182,-0.0013],[183,-0.0006],[184,0]]
       }
       middle_01_l_ctrl.Location.Y {
        type=MovieSceneScriptingFloatChannel
        default=0
        thinned_from=185
        keys [[0,0],[55,-0.0008],[57,-0.0014],[60,-0.0028],[62,-0.0032],[64,-0.0029],[68,-0.0013],[70,-0.0002],[72,-0.0015],[74,0],[75,-0.0002],[76,-0.0001],[77,-0.0003],[78,-0.0001],[84,-0.0002],[85,0],[165,0],[166,-0.001],[167,0],[168,-0.0008],[169,-0.0001],[170,0.0002],[171,0.0003],[172,0.0001],[180,-0.0001],[182,-0.0021],[183,-0.0001],[184,0]]
       }
       middle_01_l_ctrl.Location.Z {
        type=MovieSceneScriptingFloatChannel
        default=0
        thinned_from=185
        keys [[0,0],[72,-0.0002],[82,0],[83,-0.0003],[84,-0.0003],[85,0],[167,0],[168,-0.0004],[169,-0.0001],[171,-0.0004],[172,0],[180,0],[181,0.0002],[182,0.0007],[183,-0.0007],[184,0]]
       }
       middle_01_l_ctrl.Rotation.X {
        type=MovieSceneScriptingFloatChannel
        default=-8.292
        thinned_from=185
        keys [[0,-8.5995],[45,-8.5939],[53,-8.4253],[55,-8.3626],[56,-8.3021],[57,-8.1986],[60,-7.8206],[61,-7.7358],[62,-7.7032],[68,-7.721],[74,-7.7897],[76,-7.8513],[78,-7.9475],[82,-7.972],[83,-8.0439],[84,-7.9923],[86,-8.0412],[98,-8.0639],[99,-8.0808],[100,-8.1184],[106,-8.0765],[111,-8.081],[112,-8.108],[115,-8.0387],[165,-7.9799],[167,-7.8855],[168,-7.902],[169,-7.9806],[170,-8.1484],[171,-8.3349],[172,-8.4203],[173,-8.4747],[175,-8.5093],[180,-8.4904],[181,-8.4751],[182,-8.3613],[183,-8.3032],[184,-8.292]]
       }
       middle_01_l_ctrl.Rotation.Y {
        type=MovieSceneScriptingFloatChannel
        default=-9.0978
        thinned_from=185
        keys [[0,-11.2725],[2,-11.2368],[3,-11.2632],[5,-11.23],[7,-11.2572],[8,-11.2953],[11,-11.2471],[12,-11.2748],[13,-11.2404],[15,-11.296],[17,-11.3169],[18,-11.2572],[20,-11.2078],[21,-11.2543],[23,-11.2734],[24,-11.2316],[25,-11.2827],[26,-11.2674],[27,-11.2844],[28,-11.2277],[30,-11.2629],[31,-11.1548],[32,-11.2437],[34,-11.2613],[36,-11.2305],[37,-11.2467],[38,-11.1881],[39,-11.249],[41,-11.2342],[42,-11.2815],[44,-11.3236],[45,-11.2984],[46,-11.2257],[47,-11.1104],[48,-10.9572],[49,-10.7709],[50,-10.5561],[52,-10.0603],[55,-9.2217],[56,-8.6828],[57,-7.754],[58,-6.6078],[59,-5.4163],[60,-4.352],[61,-3.587],[62,-3.2938],[63,-3.3026],[64,-3.3264],[66,-3.4018],[68,-3.4863],[69,-3.5409],[70,-3.6044],[72,-3.7133],[73,-3.7765],[74,-3.8635],[75,-4.0516],[76,-4.2724],[77,-4.4757],[78,-4.7023],[79,-4.7492],[81,-4.8135],[82,-4.8604],[83,-5.5018],[84,-5.4476],[85,-5.4559],[86,-5.5799],[87,-5.5673],[88,-5.5792],[89,-5.5657],[96,-5.5687],[97,-5.5547],[98,-5.5583],[99,-5.7098],[100,-5.8123],[101,-5.7644],[102,-5.694],[104,-5.591],[106,-5.601],[107,-5.6385],[108,-5.7118],[109,-5.6411],[110,-5.6683],[112,-5.8198],[113,-5.5685],[114,-5.4947],[115,-5.4562],[116,-5.4602],[117,-5.5091],[118,-5.498],[119,-5.4559],[120,-5.5132],[121,-5.4356],[122,-5.3835],[123,-5.347],[124,-5.3373],[125,-5.4489],[126,-5.3686],[127,-5.3907],[129,-5.311],[130,-5.3308],[132,-5.4096],[133,-5.3984],[134,-5.3606],[135,-5.3113],[136,-5.398],[137,-5.3623],[138,-5.3425],[139,-5.3657],[141,-5.3589],[142,-5.3906],[143,-5.3182],[144,-5.1899],[145,-5.2402],[146,-5.1397],[147,-5.3818],[149,-5.169],[150,-5.2314],[151,-5.2329],[152,-5.2225],[153,-5.3276],[154,-5.1361],[155,-5.1284],[156,-5.2502],[157,-5.2292],[158,-5.1821],[159,-5.2054],[160,-5.2139],[161,-5.205],[162,-5.1825],[164,-5.1505],[165,-5.1172],[166,-4.7669],[167,-4.3494],[168,-4.7303],[169,-5.3304],[170,-6.821],[171,-8.5462],[172,-9.3939],[173,-10.0614],[174,-10.3414],[175,-10.4473],[176,-10.4448],[177,-10.4074],[178,-10.4912],[179,-10.4197],[180,-10.2743],[181,-10.2236],[182,-9.2986],[183,-8.8004],[184,-9.0978]]
       }
       middle_01_l_ctrl.Rotation.Z {
        type=MovieSceneScriptingFloatChannel
        default=-64.3499
        thinned_from=185
        keys [[0,-77.4826],[1,-77.5932],[5,-77.4026],[9,-77.707],[11,-77.5362],[12,-77.7005],[17,-77.9425],[19,-77.486],[22,-77.5747],[24,-77.327],[25,-77.4625],[27,-77.4188],[28,-77.2399],[30,-77.4361],[31,-77.1696],[32,-77.4576],[35,-77.4101],[36,-77.5983],[38,-77.3463],[39,-77.5472],[42,-77.6376],[43,-77.7748],[45,-77.3341],[47,-76.3167],[49,-74.8004],[51,-72.8702],[54,-69.3848],[55,-68.1081],[56,-65.6972],[57,-61.5685],[58,-56.4834],[59,-51.2031],[60,-46.4889],[61,-43.1022],[62,-41.8041],[65,-42.017],[68,-42.4854],[69,-42.8696],[70,-43.3577],[73,-44.3096],[74,-44.7232],[76,-46.8444],[77,-48.3303],[78,-50.0138],[82,-50.8826],[83,-53.5579],[84,-51.9651],[85,-52.4891],[86,-53.5191],[88,-53.907],[96,-54.1536],[97,-54.1052],[98,-54.1952],[99,-54.8353],[100,-56.0678],[102,-55.4734],[104,-54.7724],[106,-54.6061],[108,-54.8757],[110,-54.7165],[111,-54.8631],[112,-55.7475],[114,-54.0637],[115,-53.3413],[116,-53.2505],[117,-53.4585],[118,-53.4287],[119,-53.2417],[120,-53.2477],[122,-52.8012],[123,-52.9514],[124,-52.8788],[125,-53.1473],[126,-52.8811],[127,-52.84],[129,-52.5121],[130,-52.4701],[131,-52.6054],[132,-52.6078],[133,-52.4501],[134,-52.525],[135,-52.3555],[136,-52.5622],[137,-52.5542],[138,-52.428],[139,-52.555],[140,-52.453],[142,-52.5861],[143,-52.4934],[144,-52.2076],[145,-52.2722],[146,-52.0472],[147,-52.3905],[148,-52.2377],[149,-51.9553],[151,-52.0422],[152,-51.9924],[153,-52.2359],[154,-51.8189],[155,-51.7738],[156,-51.9384],[158,-51.7197],[160,-51.7236],[161,-51.831],[163,-51.6967],[165,-51.2991],[166,-49.8623],[167,-47.8949],[168,-48.6505],[169,-51.4242],[170,-57.6893],[171,-65.2083],[172,-69.0047],[173,-71.555],[174,-72.6525],[175,-73.205],[176,-72.9132],[178,-72.7795],[180,-72.3428],[181,-71.7227],[182,-67.0258],[183,-64.3704],[184,-64.3499]]
       }
       middle_01_l_ctrl.Scale.X {type=MovieSceneScriptingFloatChannel, default=1, const [1,185]}
       middle_01_l_ctrl.Scale.Y {type=MovieSceneScriptingFloatChannel, default=1, const [1,185]}
       middle_01_l_ctrl.Scale.Z {type=MovieSceneScriptingFloatChannel, default=1, const [1,185]}
       middle_02_l_ctrl.Location.X {
        type=MovieSceneScriptingFloatChannel
        default=-0.0453
        thinned_from=185
        keys [[0,-0.0452],[68,-0.045],[70,-0.0452],[72,-0.045],[82,-0.0453],[83,-0.0456],[84,-0.0456],[85,-0.0453],[165,-0.0452],[166,-0.0457],[167,-0.0452],[168,-0.0458],[170,-0.0453],[180,-0.0453],[181,-0.0454],[182,-0.0465],[183,-0.0456],[184,-0.0453]]
       }
       middle_02_l_ctrl.Location.Y {
        type=MovieSceneScriptingFloatChannel
        default=0.0507
        thinned_from=185
        keys [[0,0.0507],[57,0.0503],[68,0.05],[70,0.0506],[72,0.0502],[74,0.0507],[75,0.0505],[76,0.0506],[77,0.0501],[78,0.0505],[82,0.0503],[84,0.0504],[85,0.0507],[165,0.0507],[166,0.0488],[167,0.0507],[168,0.0493],[169,0.0499],[172,0.0506],[181,0.0507],[182,0.051],[183,0.0506],[184,0.0507]]
       }
       middle_02_l_ctrl.Location.Z {
        type=MovieSceneScriptingFloatChannel
        default=0.1146
        thinned_from=185
        keys [[0,0.1146],[76,0.1145],[77,0.1143],[78,0.1145],[84,0.1143],[85,0.1146],[165,0.1146],[166,0.1158],[167,0.1146],[168,0.1154],[170,0.1148],[179,0.1146],[181,0.115],[182,0.1144],[184,0.1146]]
       }
       middle_02_l_ctrl.Rotation.X {
        type=MovieSceneScriptingFloatChannel
        default=-5.9153
        thinned_from=185
        keys [[0,-7.6439],[1,-7.6289],[3,-7.6271],[5,-7.6463],[9,-7.62],[10,-7.648],[11,-7.652],[12,-7.6254],[13,-7.6178],[14,-7.6317],[16,-7.5855],[17,-7.5879],[18,-7.6284],[21,-7.6167],[22,-7.6305],[23,-7.6168],[24,-7.6293],[25,-7.6078],[26,-7.6098],[27,-7.5921],[28,-7.6038],[29,-7.5764],[30,-7.579],[31,-7.6116],[32,-7.5827],[35,-7.5889],[36,-7.5754],[37,-7.578],[38,-7.6013],[40,-7.5659],[41,-7.5779],[43,-7.5357],[44,-7.5656],[45,-7.5494],[46,-7.5028],[47,-7.4289],[48,-7.3307],[49,-7.2113],[50,-7.0736],[52,-6.7557],[55,-6.2181],[56,-5.8726],[57,-5.2772],[58,-4.5424],[59,-3.7786],[60,-3.0963],[61,-2.6059],[62,-2.4179],[63,-2.4225],[67,-2.4801],[68,-2.4884],[70,-2.4942],[72,-2.516],[74,-2.5487],[76,-2.5624],[77,-2.6764],[78,-2.6809],[79,-2.6939],[82,-2.7679],[83,-2.7988],[84,-3.178],[85,-3.3411],[87,-3.341],[89,-3.3528],[97,-3.3451],[98,-3.2874],[99,-3.2061],[100,-3.0788],[101,-3.0495],[102,-3.0066],[103,-3.0307],[106,-3.0466],[107,-3.0397],[109,-3.0101],[111,-3.0025],[113,-2.9748],[114,-2.9842],[115,-2.9757],[116,-2.9898],[117,-2.9805],[118,-2.9791],[119,-2.9713],[120,-2.9857],[121,-2.9736],[122,-2.9686],[123,-2.937],[125,-2.9146],[128,-2.9053],[129,-2.9218],[130,-2.9262],[131,-2.9241],[132,-2.929],[133,-2.9278],[135,-2.9038],[136,-2.919],[139,-2.8845],[140,-2.8785],[141,-2.911],[142,-2.911],[143,-2.8806],[144,-2.8434],[145,-2.8177],[147,-2.8577],[148,-2.8629],[149,-2.8406],[150,-2.87],[151,-2.8683],[152,-2.8847],[153,-2.8709],[154,-2.8387],[155,-2.8355],[156,-2.881],[157,-2.8846],[158,-2.8731],[160,-2.8672],[161,-2.8569],[163,-2.8594],[164,-2.8557],[165,-2.8317],[166,-2.7635],[167,-2.585],[168,-2.6807],[169,-2.9363],[170,-3.6399],[171,-4.7405],[172,-5.3143],[173,-5.7701],[174,-5.9939],[175,-6.1871],[179,-6.3928],[180,-6.328],[182,-6.3507],[183,-5.6635],[184,-5.9153]]
       }
       middle_02_l_ctrl.Rotation.Y {
        type=MovieSceneScriptingFloatChannel
        default=6.2924
        thinned_from=185
        keys [[0,6.4517],[45,6.4466],[47,6.4009],[49,6.3191],[52,6.1482],[55,5.9463],[56,5.8164],[57,5.5927],[58,5.3166],[59,5.0296],[60,4.7733],[61,4.589],[62,4.5184],[68,4.5802],[71,4.593],[74,4.6324],[76,4.6446],[77,4.742],[79,4.7531],[83,4.8372],[84,5.116],[85,5.2272],[97,5.2298],[98,5.1915],[99,5.1367],[100,5.048],[101,5.027],[102,4.9958],[103,5.0134],[106,5.0249],[113,4.9724],[114,4.9794],[115,4.9731],[116,4.9835],[119,4.9699],[120,4.9805],[122,4.9679],[123,4.9443],[125,4.9274],[128,4.9204],[130,4.9362],[133,4.9374],[135,4.9192],[136,4.9307],[140,4.8999],[141,4.9247],[142,4.9247],[145,4.8529],[147,4.884],[148,4.8879],[149,4.8708],[150,4.8935],[151,4.8921],[152,4.9047],[153,4.8941],[154,4.8692],[155,4.8668],[156,4.9018],[157,4.9046],[162,4.8824],[164,4.8824],[165,4.8638],[166,4.8168],[167,4.6634],[168,4.7409],[169,4.9391],[170,5.4034],[171,5.9392],[172,6.1393],[173,6.2593],[174,6.3082],[175,6.344],[179,6.3765],[182,6.3721],[183,6.2361],[184,6.2924]]
       }
       middle_02_l_ctrl.Rotation.Z {
        type=MovieSceneScriptingFloatChannel
        default=-23.2852
        thinned_from=185
        keys [[0,-36.5212],[1,-36.4071],[3,-36.3932],[4,-36.507],[5,-36.54],[7,-36.3601],[9,-36.3398],[10,-36.5524],[11,-36.5834],[12,-36.3804],[13,-36.3225],[14,-36.4284],[16,-36.0772],[17,-36.0951],[18,-36.4034],[19,-36.4229],[21,-36.3142],[22,-36.4195],[23,-36.3152],[24,-36.4104],[25,-36.2463],[26,-36.2617],[27,-36.1269],[28,-36.2165],[29,-36.008],[30,-36.0277],[31,-36.2754],[32,-36.0556],[35,-36.1028],[36,-36.0001],[37,-36.0197],[38,-36.1972],[39,-36.0053],[40,-35.928],[41,-36.019],[43,-35.6983],[44,-35.9255],[45,-35.7908],[46,-35.4035],[47,-34.789],[48,-33.9724],[49,-32.979],[50,-31.8341],[51,-30.5629],[52,-29.1906],[54,-26.2442],[55,-24.7204],[56,-21.8473],[57,-16.8963],[58,-10.7859],[59,-4.4345],[60,1.2394],[61,5.3173],[62,6.8807],[63,6.8334],[64,6.7134],[66,6.3857],[67,6.2434],[68,6.1592],[70,6.1],[71,6.0151],[72,5.8838],[74,5.5519],[75,5.473],[76,5.415],[77,4.287],[78,4.2416],[79,4.1129],[80,3.8958],[81,3.6394],[82,3.3931],[83,3.1172],[84,-0.4578],[85,-1.9786],[87,-1.9773],[88,-2.0567],[89,-2.0845],[90,-2.0669],[92,-2.0502],[93,-2.053],[96,-2.0506],[97,-2.0151],[98,-1.4869],[99,-0.7412],[100,0.436],[101,0.7094],[102,1.112],[103,0.8854],[104,0.8165],[105,0.7915],[106,0.7364],[107,0.8004],[108,0.9456],[109,1.079],[111,1.1501],[112,1.266],[113,1.4118],[114,1.3228],[115,1.4032],[116,1.2703],[117,1.3575],[118,1.371],[119,1.4448],[120,1.3089],[121,1.4228],[122,1.4702],[123,1.7695],[124,1.8793],[125,1.9834],[126,1.9875],[127,2.0204],[128,2.0711],[129,1.9146],[130,1.872],[131,1.8921],[132,1.8456],[133,1.857],[134,1.9651],[135,2.086],[136,1.9415],[137,2.0532],[138,2.1543],[139,2.2694],[140,2.3277],[141,2.0171],[142,2.0166],[143,2.3072],[144,2.6633],[145,2.9109],[147,2.5261],[148,2.4769],[149,2.6902],[150,2.408],[151,2.4245],[152,2.2676],[153,2.4005],[154,2.7093],[155,2.7399],[156,2.3036],[157,2.2691],[158,2.3788],[159,2.4127],[160,2.4351],[161,2.5343],[162,2.5456],[163,2.5097],[164,2.5457],[165,2.7762],[166,3.4412],[167,5.1902],[168,4.2423],[169,1.7771],[170,-4.649],[171,-13.9686],[172,-18.5698],[173,-22.1567],[174,-23.8956],[175,-25.3906],[179,-26.9742],[180,-26.4777],[182,-26.6833],[183,-21.3294],[184,-23.2852]]
       }
       middle_02_l_ctrl.Scale.X {type=MovieSceneScriptingFloatChannel, default=1, const [1,185]}
       middle_02_l_ctrl.Scale.Y {type=MovieSceneScriptingFloatChannel, default=1, const [1,185]}
       middle_02_l_ctrl.Scale.Z {type=MovieSceneScriptingFloatChannel, default=1, const [1,185]}
       middle_03_l_ctrl.Location.X {
        type=MovieSceneScriptingFloatChannel
        default=0.0206
        thinned_from=185
        keys [[0,0.0206],[72,0.0204],[82,0.0206],[84,0.0204],[165,0.0206],[166,0.0196],[167,0.0206],[168,0.0202],[170,0.0205],[179,0.0206],[180,0.0205],[181,0.0196],[182,0.0194],[183,0.0204],[184,0.0206]]
       }
       middle_03_l_ctrl.Location.Y {
        type=MovieSceneScriptingFloatChannel
        default=-0.0583
        thinned_from=185
        keys [[0,-0.0583],[72,-0.058],[165,-0.0583],[166,-0.0585],[167,-0.0583],[168,-0.0584],[181,-0.0583],[182,-0.0591],[183,-0.0585],[184,-0.0583]]
       }
       middle_03_l_ctrl.Location.Z {
        type=MovieSceneScriptingFloatChannel
        default=0.1324
        thinned_from=185
        keys [[0,0.1324],[76,0.1324],[77,0.1323],[165,0.1324],[166,0.1336],[167,0.1324],[168,0.133],[170,0.1322],[184,0.1324]]
       }
       middle_03_l_ctrl.Rotation.X {
        type=MovieSceneScriptingFloatChannel
        default=-2.9511
        thinned_from=185
        keys [[0,-3.9094],[1,-3.9193],[2,-3.9025],[3,-3.9133],[7,-3.912],[8,-3.9034],[9,-3.921],[10,-3.9203],[11,-3.9043],[12,-3.9071],[13,-3.8977],[14,-3.9212],[15,-3.8993],[17,-3.8907],[18,-3.9014],[20,-3.8884],[24,-3.8979],[27,-3.8938],[28,-3.8794],[30,-3.8834],[31,-3.8783],[32,-3.882],[33,-3.8766],[34,-3.8863],[38,-3.8869],[39,-3.8741],[42,-3.8824],[43,-3.8697],[44,-3.8793],[45,-3.8712],[46,-3.8479],[47,-3.8109],[48,-3.7617],[49,-3.7019],[50,-3.633],[52,-3.4738],[55,-3.2047],[56,-3.0317],[57,-2.7336],[58,-2.3656],[59,-1.9832],[60,-1.6416],[61,-1.3961],[62,-1.3019],[67,-1.3177],[70,-1.317],[71,-1.3222],[75,-1.3679],[76,-1.3872],[77,-1.4375],[78,-1.44],[79,-1.448],[80,-1.4625],[82,-1.5009],[83,-1.5379],[84,-1.7088],[85,-1.7976],[86,-1.8503],[87,-1.861],[88,-1.8783],[89,-1.8869],[96,-1.8929],[97,-1.8818],[101,-1.8258],[102,-1.8048],[103,-1.8226],[104,-1.8225],[105,-1.806],[106,-1.8279],[109,-1.8185],[110,-1.8192],[111,-1.8287],[112,-1.949],[113,-1.8535],[114,-1.7728],[115,-1.6421],[116,-1.6504],[117,-1.6554],[119,-1.644],[120,-1.6485],[122,-1.6388],[123,-1.6124],[124,-1.6061],[126,-1.6194],[128,-1.6199],[129,-1.6168],[130,-1.6263],[131,-1.6256],[132,-1.6336],[133,-1.6378],[134,-1.6084],[135,-1.6062],[136,-1.6222],[137,-1.6225],[139,-1.6101],[140,-1.6085],[142,-1.6197],[143,-1.596],[144,-1.5661],[145,-1.5474],[146,-1.5692],[147,-1.5669],[148,-1.5831],[149,-1.579],[150,-1.5876],[151,-1.5876],[152,-1.5961],[153,-1.59],[154,-1.5673],[155,-1.5699],[156,-1.6],[157,-1.5964],[158,-1.5893],[159,-1.586],[160,-1.5922],[161,-1.6102],[163,-1.6276],[165,-1.6322],[166,-1.5523],[167,-1.4533],[168,-1.4914],[169,-1.6412],[170,-2.006],[171,-2.5555],[172,-2.8677],[173,-3.1041],[175,-3.2456],[176,-3.2338],[177,-3.259],[178,-3.3111],[179,-3.3273],[180,-3.2365],[181,-3.2431],[182,-3.1734],[183,-2.804],[184,-2.9511]]
       }
       middle_03_l_ctrl.Rotation.Y {
        type=MovieSceneScriptingFloatChannel
        default=5.0873
        thinned_from=185
        keys [[0,5.7],[1,5.7053],[2,5.6963],[8,5.6968],[10,5.7058],[13,5.6937],[14,5.7063],[17,5.6898],[27,5.6916],[33,5.6822],[38,5.6878],[43,5.6785],[44,5.6837],[45,5.6765],[46,5.6556],[48,5.5786],[50,5.4635],[53,5.2432],[55,5.0805],[56,4.9258],[57,4.6593],[58,4.3303],[59,3.9884],[60,3.683],[61,3.4634],[62,3.3793],[68,3.4037],[70,3.4009],[71,3.4081],[75,3.4725],[76,3.499],[77,3.5676],[78,3.5708],[79,3.5813],[82,3.6514],[83,3.6972],[84,3.9063],[85,4.0128],[86,4.0735],[89,4.1147],[96,4.1213],[101,4.0462],[102,4.0223],[103,4.0426],[104,4.0425],[105,4.0237],[106,4.0487],[110,4.0387],[111,4.0496],[112,4.1827],[113,4.0775],[114,3.9855],[115,3.8296],[117,3.8459],[119,3.832],[120,3.8374],[122,3.8256],[123,3.7931],[124,3.7852],[126,3.8017],[129,3.7985],[130,3.8102],[131,3.8093],[133,3.8244],[134,3.788],[135,3.7853],[136,3.8051],[138,3.7993],[140,3.7882],[142,3.8021],[143,3.7726],[144,3.735],[145,3.7112],[146,3.7389],[147,3.736],[148,3.7564],[149,3.7512],[150,3.762],[151,3.7621],[152,3.7728],[153,3.765],[154,3.7365],[155,3.7398],[156,3.7776],[159,3.76],[160,3.7678],[161,3.7902],[163,3.8119],[165,3.8174],[166,3.718],[167,3.5884],[168,3.6387],[169,3.8279],[170,4.2422],[171,4.7691],[172,5.0238],[173,5.1995],[175,5.2977],[176,5.2897],[177,5.3068],[178,5.3416],[179,5.3524],[180,5.2917],[181,5.2966],[182,5.2551],[183,4.9732],[184,5.0873]]
       }
       middle_03_l_ctrl.Rotation.Z {
        type=MovieSceneScriptingFloatChannel
        default=-24.9329
        thinned_from=185
        keys [[0,-33.5854],[1,-33.6704],[2,-33.5256],[3,-33.6187],[7,-33.6073],[8,-33.5335],[9,-33.6855],[10,-33.6793],[11,-33.5414],[12,-33.5655],[13,-33.4846],[14,-33.6868],[15,-33.498],[17,-33.4234],[18,-33.516],[20,-33.4034],[24,-33.4865],[27,-33.4508],[28,-33.3265],[30,-33.3606],[31,-33.317],[32,-33.3488],[33,-33.3021],[34,-33.3861],[38,-33.3911],[39,-33.2803],[42,-33.3517],[43,-33.242],[44,-33.3256],[45,-33.2422],[46,-33.0027],[47,-32.6225],[48,-32.1173],[49,-31.5028],[50,-30.7945],[51,-30.0081],[53,-28.2634],[55,-26.3938],[56,-24.6164],[57,-21.5536],[58,-17.7735],[59,-13.8443],[60,-10.3343],[61,-7.8116],[62,-6.8444],[63,-6.8614],[67,-7.0566],[68,-7.0736],[70,-7.0474],[71,-7.1096],[72,-7.2375],[74,-7.5723],[75,-7.7095],[76,-7.9589],[77,-8.5951],[78,-8.6327],[79,-8.7378],[80,-8.92],[82,-9.3929],[83,-9.843],[84,-11.8898],[85,-12.9465],[86,-13.5559],[87,-13.6768],[88,-13.8761],[89,-13.9742],[90,-14.0172],[96,-14.0414],[97,-13.916],[98,-13.741],[100,-13.4472],[101,-13.2775],[102,-13.0362],[103,-13.2413],[104,-13.2399],[105,-13.0508],[106,-13.3024],[107,-13.2847],[109,-13.194],[110,-13.2019],[111,-13.3129],[112,-14.6758],[113,-13.5947],[114,-12.6673],[115,-11.1243],[116,-11.2243],[117,-11.284],[119,-11.1479],[120,-11.2012],[121,-11.1316],[122,-11.0851],[123,-10.7677],[124,-10.6916],[126,-10.852],[128,-10.8577],[129,-10.8204],[130,-10.9346],[131,-10.9266],[132,-11.023],[133,-11.0733],[134,-10.7184],[135,-10.6924],[136,-10.8853],[137,-10.8891],[138,-10.8282],[139,-10.74],[140,-10.7204],[142,-10.8555],[143,-10.5692],[144,-10.2047],[145,-9.9756],[146,-10.2423],[147,-10.2145],[148,-10.4118],[149,-10.3616],[150,-10.4662],[151,-10.4666],[152,-10.5706],[153,-10.4955],[154,-10.2195],[155,-10.2516],[156,-10.6178],[157,-10.5742],[158,-10.4873],[159,-10.447],[160,-10.5221],[161,-10.7403],[162,-10.829],[163,-10.9511],[164,-10.9648],[165,-11.0055],[166,-10.0016],[167,-8.8028],[168,-9.2922],[169,-11.1357],[170,-15.3352],[171,-21.108],[172,-24.1417],[173,-26.3796],[175,-27.6837],[176,-27.5745],[177,-27.8062],[178,-28.2811],[179,-28.43],[180,-27.6014],[181,-27.6649],[182,-27.0411],[183,-23.521],[184,-24.9329]]
       }
       middle_03_l_ctrl.Scale.X {type=MovieSceneScriptingFloatChannel, default=1, const [1,185]}
       middle_03_l_ctrl.Scale.Y {type=MovieSceneScriptingFloatChannel, default=1, const [1,185]}
       middle_03_l_ctrl.Scale.Z {type=MovieSceneScriptingFloatChannel, default=1, const [1,185]}
       ring_metacarpal_l_ctrl.Location.X {
        type=MovieSceneScriptingFloatChannel
        default=0
        thinned_from=185
        keys [[0,0],[47,0],[56,-0.0006],[61,-0.0017],[64,-0.0017],[68,-0.001],[70,-0.0002],[72,-0.001],[74,0],[75,-0.0003],[76,-0.0002],[77,-0.0008],[78,-0.0002],[82,-0.0004],[83,-0.0002],[165,0],[166,-0.0016],[167,0],[168,-0.0015],[169,-0.0011],[171,-0.0007],[172,-0.0001],[179,0],[180,-0.0002],[181,-0.001],[182,-0.0005],[183,-0.0004],[184,0]]
       }
       ring_metacarpal_l_ctrl.Location.Y {
        type=MovieSceneScriptingFloatChannel
        default=0
        thinned_from=185
        keys [[0,0],[64,-0.0005],[70,0],[76,0.0001],[77,0.0003],[78,0.0001],[82,0.0002],[165,0],[166,-0.0002],[167,0],[168,-0.0002],[171,0.0002],[180,0],[182,-0.0005],[183,0.0001],[184,0]]
       }
       ring_metacarpal_l_ctrl.Location.Z {
        type=MovieSceneScriptingFloatChannel
        default=0.0001
        thinned_from=185
        keys [[0,0.0001],[57,-0.0003],[64,-0.0008],[68,-0.0004],[70,0],[72,-0.0005],[74,0.0001],[76,0],[77,-0.0002],[84,-0.0001],[165,0.0001],[166,-0.0005],[167,0.0001],[168,-0.0003],[169,-0.0002],[171,-0.0002],[180,0],[181,-0.0003],[183,-0.0003],[184,0.0001]]
       }
       ring_metacarpal_l_ctrl.Rotation.X {type=MovieSceneScriptingFloatChannel, default=-0.044, const [-0.044,185]}
       ring_metacarpal_l_ctrl.Rotation.Y {type=MovieSceneScriptingFloatChannel, default=3.1497, const [3.1497,185]}
       ring_metacarpal_l_ctrl.Rotation.Z {type=MovieSceneScriptingFloatChannel, default=-0.8002, const [-0.8002,185]}
       ring_metacarpal_l_ctrl.Scale.X {type=MovieSceneScriptingFloatChannel, default=1, const [1,185]}
       ring_metacarpal_l_ctrl.Scale.Y {type=MovieSceneScriptingFloatChannel, default=1, const [1,185]}
       ring_metacarpal_l_ctrl.Scale.Z {type=MovieSceneScriptingFloatChannel, default=1, const [1,185]}
       ring_01_l_ctrl.Location.X {
        type=MovieSceneScriptingFloatChannel
        default=0
        thinned_from=185
        keys [[0,0],[50,-0.0004],[55,-0.001],[57,-0.0018],[60,-0.0035],[61,-0.0038],[62,-0.004],[65,-0.0035],[68,-0.0025],[70,-0.0004],[72,-0.0024],[74,0],[75,-0.0007],[76,-0.0004],[77,-0.0022],[78,-0.0005],[82,-0.0013],[83,-0.0003],[84,-0.0003],[85,0],[165,0],[166,-0.0044],[167,0],[168,-0.0041],[169,-0.0029],[171,-0.0018],[172,-0.0002],[179,0],[180,-0.0005],[181,-0.0027],[182,-0.0009],[183,-0.0006],[184,0]]
       }
       ring_01_l_ctrl.Location.Y {
        type=MovieSceneScriptingFloatChannel
        default=0
        thinned_from=185
        keys [[0,0],[50,-0.0003],[55,-0.001],[57,-0.0016],[60,-0.0032],[61,-0.0036],[62,-0.0037],[64,-0.0034],[68,-0.0016],[69,-0.0008],[70,-0.0003],[72,-0.0018],[74,0],[75,-0.0003],[76,-0.0001],[77,-0.0005],[78,-0.0001],[84,-0.0003],[85,0],[165,0],[166,-0.0018],[167,0],[168,-0.0017],[169,-0.0007],[171,-0.0001],[179,0],[180,-0.0001],[181,-0.0015],[182,-0.0021],[183,-0.0003],[184,0]]
       }
       ring_01_l_ctrl.Location.Z {
        type=MovieSceneScriptingFloatChannel
        default=0
        thinned_from=185
        keys [[0,0],[82,0],[83,-0.0002],[84,-0.0003],[85,0],[167,0],[168,-0.0002],[169,-0.0001],[170,-0.0004],[171,-0.0004],[172,-0.0001],[180,0],[182,0.0006],[183,-0.0007],[184,0]]
       }
       ring_01_l_ctrl.Rotation.X {
        type=MovieSceneScriptingFloatChannel
        default=-12.1159
        thinned_from=185
        keys [[0,-11.5383],[2,-11.5567],[3,-11.5282],[5,-11.5517],[6,-11.5263],[7,-11.5432],[10,-11.5081],[11,-11.5377],[17,-11.5164],[18,-11.5477],[24,-11.5468],[26,-11.5284],[29,-11.5686],[30,-11.5473],[37,-11.5408],[39,-11.5531],[42,-11.5441],[43,-11.5058],[55,-11.7723],[56,-11.8263],[60,-12.2281],[61,-12.2979],[62,-12.3246],[68,-12.2902],[70,-12.217],[73,-12.2004],[76,-12.1048],[77,-12.0246],[83,-11.9687],[84,-11.7686],[85,-11.914],[86,-11.9942],[90,-12.0748],[96,-12.0817],[97,-11.9994],[98,-11.8883],[99,-11.8214],[100,-12.11],[101,-12.0957],[102,-12.111],[104,-12.0871],[105,-12.1142],[108,-12.1027],[109,-12.1373],[110,-12.0932],[111,-12.0799],[112,-12.2098],[113,-12.2747],[114,-12.1385],[115,-11.9518],[117,-11.9191],[118,-11.9443],[120,-11.9094],[122,-11.9188],[123,-11.9553],[125,-11.9289],[126,-11.961],[130,-11.9345],[131,-11.9752],[133,-11.9827],[134,-11.968],[135,-11.9842],[136,-11.9514],[138,-11.9707],[139,-11.9586],[141,-11.9838],[142,-11.9455],[144,-11.9678],[146,-11.9593],[147,-11.9117],[149,-11.9897],[150,-11.9479],[152,-11.9683],[153,-11.939],[155,-11.9473],[156,-11.9166],[158,-11.9571],[160,-11.9483],[163,-11.9907],[165,-11.9864],[166,-12.0077],[167,-12.1575],[168,-12.3496],[169,-12.6111],[170,-12.4952],[171,-12.2485],[172,-12.1103],[173,-11.9177],[174,-11.7726],[175,-11.7777],[176,-11.7479],[180,-11.7914],[181,-11.7604],[182,-11.8381],[183,-12.0396],[184,-12.1159]]
       }
       ring_01_l_ctrl.Rotation.Y {
        type=MovieSceneScriptingFloatChannel
        default=-3.1034
        thinned_from=185
        keys [[0,-4.3797],[1,-4.4703],[2,-4.435],[3,-4.3898],[5,-4.42],[6,-4.3918],[7,-4.4109],[8,-4.4625],[9,-4.4076],[10,-4.4168],[12,-4.4476],[13,-4.4032],[14,-4.4262],[15,-4.5162],[16,-4.4441],[18,-4.5244],[19,-4.4237],[21,-4.4151],[22,-4.5506],[23,-4.3724],[24,-4.4251],[25,-4.3481],[26,-4.4343],[27,-4.3879],[28,-4.5765],[29,-4.4197],[30,-4.4848],[31,-4.4507],[32,-4.5264],[33,-4.5093],[34,-4.4498],[35,-4.4449],[36,-4.4667],[37,-4.4535],[38,-4.4899],[39,-4.4984],[40,-4.4032],[41,-4.3845],[42,-4.4492],[43,-4.0092],[44,-4.4437],[45,-4.4339],[46,-4.4059],[47,-4.3615],[48,-4.3025],[49,-4.2307],[50,-4.1479],[52,-3.9569],[55,-3.6337],[56,-3.4245],[57,-3.0619],[58,-2.6128],[59,-2.1438],[60,-1.7214],[61,-1.4124],[62,-1.2834],[64,-1.2511],[65,-1.2373],[66,-1.2263],[67,-1.219],[68,-1.2163],[69,-1.2244],[70,-1.2423],[71,-1.3011],[72,-1.3846],[73,-1.4458],[74,-1.5179],[75,-1.6567],[76,-1.806],[77,-2.0329],[78,-2.229],[79,-2.2754],[81,-2.3389],[82,-2.3854],[83,-2.92],[85,-2.596],[86,-2.6907],[87,-2.6905],[88,-2.6555],[89,-2.6367],[96,-2.6281],[97,-2.6925],[98,-2.6934],[99,-2.6781],[100,-2.8842],[101,-2.7817],[102,-2.7779],[103,-2.7272],[104,-2.7492],[105,-2.7807],[106,-2.7462],[108,-2.6583],[109,-2.7449],[110,-2.6523],[111,-2.5758],[112,-2.6053],[113,-2.7187],[114,-2.6641],[115,-2.6687],[116,-2.6448],[117,-2.6494],[118,-2.6627],[119,-2.6649],[120,-2.6173],[121,-2.6414],[122,-2.5798],[123,-2.7187],[124,-2.7196],[125,-2.6854],[126,-2.6663],[128,-2.6089],[129,-2.6009],[130,-2.4946],[131,-2.5583],[132,-2.4955],[133,-2.4893],[134,-2.5806],[135,-2.5761],[137,-2.5356],[138,-2.5228],[139,-2.459],[140,-2.5112],[141,-2.4965],[142,-2.461],[143,-2.5326],[144,-2.6261],[145,-2.7],[146,-2.6203],[147,-2.6231],[148,-2.5634],[149,-2.5881],[150,-2.5473],[151,-2.5311],[152,-2.4813],[153,-2.5287],[154,-2.5065],[155,-2.4987],[156,-2.425],[157,-2.464],[158,-2.462],[159,-2.4471],[161,-2.3957],[162,-2.3829],[163,-2.3601],[164,-2.3009],[165,-2.2486],[166,-2.1411],[167,-2.0411],[168,-1.7998],[169,-1.8161],[170,-2.3131],[171,-3.2529],[172,-3.7426],[173,-3.6893],[174,-3.8061],[175,-4.117],[176,-4.056],[177,-4.1524],[178,-4.0445],[179,-3.8781],[180,-4.0274],[181,-3.9259],[182,-3.0394],[183,-3.0189],[184,-3.1034]]
       }
       ring_01_l_ctrl.Rotation.Z {
        type=MovieSceneScriptingFloatChannel
        default=-51.8927
        thinned_from=185
        keys [[0,-66.2577],[2,-66.3183],[3,-66.1658],[5,-66.0744],[6,-66.3033],[7,-66.19],[9,-66.3556],[11,-66.1593],[12,-66.4748],[15,-66.5665],[16,-66.6861],[19,-66.2146],[21,-66.3084],[22,-66.1591],[23,-66.2504],[24,-66.1665],[25,-66.3616],[27,-66.2133],[28,-65.5264],[29,-66.0509],[30,-66.1929],[31,-65.9071],[34,-66.1836],[35,-66.1524],[36,-66.3202],[38,-66.0306],[39,-66.39],[42,-66.3825],[43,-67.2597],[51,-61.7191],[53,-60.0622],[54,-59.0901],[55,-57.9967],[56,-55.9717],[57,-52.6164],[58,-48.5293],[59,-44.3087],[60,-40.5531],[61,-37.8609],[62,-36.8305],[68,-36.8574],[69,-37.2087],[70,-37.7366],[73,-38.5599],[74,-38.9047],[76,-41.1279],[77,-42.6507],[78,-44.5285],[82,-45.1834],[83,-47.4675],[84,-45.4173],[85,-45.5282],[86,-46.4137],[87,-46.6506],[96,-46.9208],[97,-47.1167],[98,-47.5527],[99,-48.4269],[100,-49.6639],[102,-48.9999],[103,-48.3816],[104,-48.1952],[105,-48.2149],[106,-47.9919],[108,-48.1548],[109,-48.0264],[110,-48.0155],[111,-48.1288],[112,-49.0586],[113,-48.4661],[114,-47.4332],[115,-46.6722],[116,-46.5309],[117,-46.6393],[118,-46.6483],[120,-46.3139],[121,-46.2383],[122,-45.9592],[123,-46.3677],[125,-46.3092],[126,-46.0742],[130,-45.5968],[131,-45.7585],[133,-45.4593],[134,-45.7615],[135,-45.5278],[136,-45.5401],[137,-45.6836],[138,-45.4297],[139,-45.5502],[140,-45.4539],[142,-45.6127],[144,-45.4887],[145,-45.5078],[146,-45.3537],[147,-45.4012],[148,-45.3441],[150,-45.0799],[151,-45.1434],[152,-44.986],[153,-45.1161],[157,-44.7254],[158,-44.7557],[160,-44.6266],[161,-44.7562],[163,-44.6293],[165,-44.1859],[166,-43.0935],[167,-41.3219],[168,-41.639],[169,-43.9268],[170,-48.31],[171,-53.9523],[172,-57.3472],[173,-59.4993],[174,-60.8321],[175,-61.012],[176,-60.6093],[178,-60.296],[180,-59.6258],[181,-59.0004],[182,-54.0792],[183,-51.6051],[184,-51.8927]]
       }
       ring_01_l_ctrl.Scale.X {type=MovieSceneScriptingFloatChannel, default=1, const [1,185]}
       ring_01_l_ctrl.Scale.Y {type=MovieSceneScriptingFloatChannel, default=1, const [1,185]}
       ring_01_l_ctrl.Scale.Z {type=MovieSceneScriptingFloatChannel, default=1, const [1,185]}
       ring_02_l_ctrl.Location.X {
        type=MovieSceneScriptingFloatChannel
        default=0
        thinned_from=185
        keys [[0,0],[82,-0.0001],[84,-0.0004],[85,0],[165,0],[166,-0.0008],[167,0],[168,-0.0008],[170,-0.0003],[181,-0.0001],[182,-0.0008],[184,0]]
       }
       ring_02_l_ctrl.Location.Y {
        type=MovieSceneScriptingFloatChannel
        default=0
        thinned_from=185
        keys [[0,0],[57,-0.0004],[68,-0.0007],[70,-0.0001],[72,-0.0006],[74,0],[75,-0.0002],[76,-0.0001],[77,-0.0007],[78,-0.0002],[82,-0.0004],[84,-0.0003],[85,0],[165,0],[166,-0.0016],[167,0],[168,-0.0013],[169,-0.0007],[172,-0.0001],[181,-0.0002],[182,0.0003],[183,0],[184,0]]
       }
       ring_02_l_ctrl.Location.Z {
        type=MovieSceneScriptingFloatChannel
        default=0
        thinned_from=185
        keys [[0,0],[76,0],[77,-0.0001],[82,-0.0001],[84,-0.0003],[85,0],[165,0],[166,0.0014],[167,0],[168,0.0009],[170,0.0003],[179,0],[180,0.0001],[181,0.0005],[182,-0.0002],[184,0]]
       }
       ring_02_l_ctrl.Rotation.X {
        type=MovieSceneScriptingFloatChannel
        default=-8.9678
        thinned_from=185
        keys [[0,-11.4112],[2,-11.3876],[5,-11.4158],[9,-11.3822],[11,-11.4322],[12,-11.3726],[13,-11.3683],[14,-11.4017],[16,-11.3224],[17,-11.3852],[19,-11.3858],[20,-11.3573],[21,-11.3534],[22,-11.4016],[23,-11.3251],[24,-11.3543],[25,-11.3134],[27,-11.3462],[28,-11.4535],[29,-11.3353],[31,-11.3784],[33,-11.3695],[34,-11.3343],[36,-11.3147],[38,-11.3734],[39,-11.3225],[41,-11.2888],[42,-11.3137],[43,-11.0683],[44,-11.2785],[45,-11.2547],[46,-11.1864],[47,-11.0778],[48,-10.9337],[49,-10.7583],[50,-10.5561],[52,-10.0893],[55,-9.3],[56,-8.7927],[57,-7.9185],[58,-6.8396],[59,-5.7181],[60,-4.7162],[61,-3.9962],[62,-3.7201],[63,-3.7279],[64,-3.7482],[68,-3.8664],[72,-3.9315],[73,-3.9762],[74,-4.0106],[75,-4.0142],[76,-4.0551],[77,-4.2161],[78,-4.3208],[79,-4.3563],[81,-4.4049],[82,-4.4405],[83,-4.5767],[84,-5.1827],[85,-5.4441],[86,-5.4981],[89,-5.5392],[96,-5.5405],[97,-5.4859],[98,-5.3842],[100,-5.0335],[101,-4.9815],[102,-4.9436],[103,-4.9546],[104,-4.9824],[105,-4.9635],[106,-4.9902],[108,-4.9761],[109,-4.98],[110,-4.9616],[111,-4.9595],[112,-4.9064],[113,-4.9072],[114,-4.9606],[115,-4.9508],[116,-4.9691],[118,-4.9108],[120,-4.9313],[121,-4.9275],[122,-4.9123],[123,-4.8804],[125,-4.8535],[126,-4.8942],[127,-4.8663],[128,-4.8849],[131,-4.8821],[132,-4.866],[133,-4.8762],[134,-4.8559],[136,-4.8836],[138,-4.8133],[139,-4.7863],[140,-4.8255],[141,-4.8411],[142,-4.8308],[144,-4.773],[146,-4.8008],[148,-4.8181],[149,-4.8049],[150,-4.8298],[151,-4.8196],[152,-4.8297],[153,-4.8239],[154,-4.7909],[155,-4.8053],[156,-4.8479],[157,-4.8728],[159,-4.8254],[160,-4.7942],[161,-4.7912],[162,-4.8156],[163,-4.7962],[164,-4.8049],[165,-4.771],[166,-4.6371],[167,-4.4026],[168,-4.547],[169,-4.9884],[170,-5.9774],[171,-7.5968],[172,-8.3551],[173,-9.1796],[174,-9.4154],[175,-9.6905],[176,-9.7827],[177,-9.7758],[178,-9.8793],[179,-9.9378],[180,-9.6031],[181,-9.644],[182,-9.5881],[183,-8.676],[184,-8.9678]]
       }
       ring_02_l_ctrl.Rotation.Y {
        type=MovieSceneScriptingFloatChannel
        default=6.0278
        thinned_from=185
        keys [[0,5.7606],[10,5.7621],[11,5.7553],[12,5.77],[14,5.7629],[16,5.782],[17,5.767],[21,5.7746],[22,5.7629],[23,5.7814],[24,5.7745],[25,5.7842],[27,5.7765],[28,5.75],[29,5.779],[33,5.7708],[36,5.7839],[38,5.7698],[41,5.79],[42,5.7845],[43,5.8387],[52,5.493],[54,5.3945],[55,5.3358],[56,5.2284],[57,5.0516],[59,4.615],[60,4.4179],[61,4.2767],[62,4.2226],[63,4.2282],[68,4.3282],[72,4.3749],[74,4.4283],[75,4.4312],[76,4.459],[77,4.5674],[78,4.6318],[82,4.7062],[83,4.7852],[84,5.1055],[85,5.2297],[86,5.2539],[96,5.2726],[97,5.2486],[98,5.2028],[100,5.0343],[102,4.988],[104,5.0081],[105,4.9983],[106,5.0121],[109,5.0069],[111,4.9963],[112,4.9686],[113,4.9689],[114,4.9968],[115,4.9917],[116,5.0012],[118,4.9708],[121,4.9795],[125,4.9404],[126,4.962],[127,4.9472],[128,4.9571],[131,4.9556],[132,4.947],[133,4.9524],[134,4.9416],[136,4.9563],[139,4.904],[140,4.9253],[141,4.9337],[142,4.9281],[144,4.8966],[148,4.9213],[149,4.9141],[150,4.9276],[151,4.9221],[153,4.9243],[154,4.9065],[155,4.9143],[157,4.9506],[160,4.9083],[161,4.9066],[162,4.9199],[163,4.9094],[164,4.9141],[165,4.8956],[166,4.8252],[167,4.6819],[168,4.7666],[169,5.007],[170,5.4387],[171,5.8837],[172,5.9907],[173,6.029],[179,6.0007],[180,6.0211],[183,6.0162],[184,6.0278]]
       }
       ring_02_l_ctrl.Rotation.Z {
        type=MovieSceneScriptingFloatChannel
        default=-42.1204
        thinned_from=185
        keys [[0,-57.5382],[1,-57.5341],[2,-57.3862],[5,-57.5678],[7,-57.5687],[9,-57.3518],[11,-57.6739],[12,-57.2899],[13,-57.2622],[14,-57.4775],[15,-57.294],[16,-56.9676],[17,-57.3707],[19,-57.375],[20,-57.1916],[21,-57.1665],[22,-57.4764],[23,-56.9845],[24,-57.1719],[25,-56.9096],[27,-57.1196],[28,-57.811],[29,-57.0503],[30,-57.1235],[31,-57.3274],[33,-57.2698],[34,-57.0435],[36,-56.9175],[38,-57.2948],[39,-56.9675],[41,-56.7516],[42,-56.9117],[43,-55.3395],[44,-56.6853],[45,-56.5286],[46,-56.0781],[47,-55.3632],[48,-54.4133],[49,-53.2576],[50,-51.9257],[51,-50.4469],[53,-47.1662],[55,-43.6504],[56,-40.3081],[57,-34.5485],[58,-27.4402],[59,-20.0515],[60,-13.451],[61,-8.7072],[62,-6.8885],[63,-6.9488],[64,-7.1066],[65,-7.3276],[67,-7.8207],[68,-8.0239],[69,-8.154],[71,-8.3704],[72,-8.5155],[73,-8.8568],[74,-9.1219],[75,-9.1482],[76,-9.4571],[77,-10.658],[78,-11.4335],[79,-11.6935],[81,-12.0493],[82,-12.3093],[83,-13.2787],[84,-17.5581],[85,-19.3854],[86,-19.7547],[87,-19.8144],[88,-19.9456],[89,-20.0355],[96,-20.044],[97,-19.6716],[98,-18.9731],[99,-17.7848],[100,-16.5448],[101,-16.1797],[102,-15.9143],[103,-15.9917],[104,-16.1869],[105,-16.0538],[106,-16.2418],[107,-16.2128],[108,-16.143],[109,-16.1701],[110,-16.0402],[111,-16.0261],[112,-15.6515],[113,-15.6575],[114,-16.0334],[115,-15.9648],[116,-16.0931],[118,-15.683],[119,-15.7337],[120,-15.8272],[121,-15.8004],[122,-15.6934],[123,-15.4683],[125,-15.2779],[126,-15.5656],[127,-15.3687],[128,-15.5003],[129,-15.4821],[130,-15.5098],[131,-15.4801],[132,-15.3661],[133,-15.4382],[134,-15.2949],[136,-15.49],[137,-15.2576],[138,-14.9928],[139,-14.8009],[140,-15.0793],[141,-15.1898],[142,-15.1171],[144,-14.706],[145,-14.7857],[146,-14.904],[147,-14.9394],[148,-15.0268],[149,-14.933],[150,-15.1097],[151,-15.0373],[152,-15.1088],[153,-15.0671],[154,-14.8332],[155,-14.9356],[156,-15.2383],[157,-15.4142],[158,-15.2136],[159,-15.0788],[160,-14.8568],[161,-14.8356],[162,-15.0087],[163,-14.8714],[164,-14.9331],[165,-14.6921],[166,-13.7157],[167,-12.0341],[168,-13.0905],[169,-16.2387],[170,-22.99],[171,-33.513],[172,-38.2843],[173,-43.4501],[174,-44.9213],[175,-46.6427],[176,-47.22],[177,-47.1769],[178,-47.8248],[179,-48.1925],[180,-46.0987],[181,-46.3608],[182,-46.0238],[183,-40.2975],[184,-42.1204]]
       }
       ring_02_l_ctrl.Scale.X {type=MovieSceneScriptingFloatChannel, default=1, const [1,185]}
       ring_02_l_ctrl.Scale.Y {type=MovieSceneScriptingFloatChannel, default=1, const [1,185]}
       ring_02_l_ctrl.Scale.Z {type=MovieSceneScriptingFloatChannel, default=1, const [1,185]}
       ring_03_l_ctrl.Location.X {
        type=MovieSceneScriptingFloatChannel
        default=0
        thinned_from=185
        keys [[0,0],[64,-0.0007],[70,0],[72,-0.0003],[82,0],[84,-0.0002],[165,0],[166,-0.0015],[167,0],[168,-0.0007],[169,-0.0003],[170,0],[179,0],[180,-0.0001],[181,-0.0009],[182,-0.0011],[183,-0.0002],[184,0]]
       }
       ring_03_l_ctrl.Location.Y {
        type=MovieSceneScriptingFloatChannel
        default=0
        thinned_from=185
        keys [[0,0],[64,0.0007],[68,0.0004],[70,0],[72,0.0003],[74,0],[165,0],[166,0.0001],[181,-0.0001],[182,-0.0007],[183,-0.0002],[184,0]]
       }
       ring_03_l_ctrl.Location.Z {
        type=MovieSceneScriptingFloatChannel
        default=0
        thinned_from=185
        keys [[0,0],[165,0],[166,0.0011],[167,0],[168,0.0006],[170,-0.0001],[181,0],[183,0.0001],[184,0]]
       }
       ring_03_l_ctrl.Rotation.X {
        type=MovieSceneScriptingFloatChannel
        default=-3.3576
        thinned_from=185
        keys [[0,-4.4021],[2,-4.4099],[3,-4.4228],[4,-4.4215],[6,-4.4428],[7,-4.4227],[8,-4.4185],[10,-4.4527],[11,-4.4181],[12,-4.4401],[13,-4.4335],[14,-4.4623],[15,-4.4182],[16,-4.425],[17,-4.3853],[18,-4.4127],[20,-4.4142],[21,-4.4261],[22,-4.393],[23,-4.4359],[24,-4.4174],[25,-4.4237],[26,-4.3972],[27,-4.3968],[28,-4.3115],[29,-4.3791],[31,-4.3876],[32,-4.3653],[33,-4.3703],[34,-4.3937],[35,-4.4033],[37,-4.4015],[38,-4.392],[39,-4.3929],[40,-4.4257],[42,-4.3951],[43,-4.4713],[45,-4.3353],[50,-4.0448],[52,-3.9132],[53,-3.8374],[54,-3.7526],[55,-3.6571],[56,-3.4813],[57,-3.1911],[58,-2.838],[59,-2.4735],[60,-2.1493],[61,-1.917],[62,-1.8281],[68,-1.8284],[70,-1.8099],[71,-1.8155],[72,-1.8276],[74,-1.8615],[75,-1.8723],[76,-1.8736],[77,-1.9315],[78,-1.916],[79,-1.9266],[81,-1.9762],[82,-1.9919],[83,-1.9964],[84,-2.1874],[85,-2.2656],[86,-2.2996],[87,-2.3021],[88,-2.317],[89,-2.325],[96,-2.3284],[97,-2.3498],[98,-2.341],[99,-2.3195],[100,-2.2477],[101,-2.2361],[102,-2.2115],[103,-2.2626],[104,-2.2507],[105,-2.219],[106,-2.2493],[107,-2.2467],[108,-2.2179],[109,-2.2295],[110,-2.2188],[111,-2.2265],[112,-2.337],[113,-2.2819],[114,-2.2095],[115,-2.0962],[116,-2.1105],[117,-2.1109],[118,-2.0985],[119,-2.0927],[120,-2.1005],[122,-2.1061],[123,-2.0628],[124,-2.0553],[125,-2.0318],[126,-2.0616],[127,-2.062],[128,-2.0546],[129,-2.0546],[130,-2.072],[131,-2.0545],[132,-2.056],[133,-2.0692],[134,-2.0341],[135,-2.0311],[137,-2.0537],[138,-2.0421],[139,-2.0197],[140,-2.0386],[141,-2.0353],[142,-2.0419],[144,-1.9925],[145,-1.9836],[146,-2.0101],[147,-2.0032],[148,-2.0144],[149,-2.0143],[150,-2.0283],[151,-2.0246],[152,-2.0329],[153,-2.0148],[154,-2.002],[155,-2.007],[156,-2.018],[157,-2.0158],[158,-2.0065],[159,-2.0156],[160,-2.0169],[161,-2.0326],[162,-2.0267],[163,-2.0433],[165,-2.0383],[166,-1.9914],[167,-1.9119],[168,-1.8507],[169,-1.9573],[170,-2.2927],[171,-2.8501],[172,-3.1846],[173,-3.5124],[174,-3.68],[175,-3.7431],[177,-3.7365],[178,-3.8233],[179,-3.8072],[180,-3.6275],[181,-3.6679],[182,-3.6393],[183,-3.2134],[184,-3.3576]]
       }
       ring_03_l_ctrl.Rotation.Y {
        type=MovieSceneScriptingFloatChannel
        default=4.0428
        thinned_from=185
        keys [[0,4.7391],[6,4.7625],[8,4.7486],[10,4.7681],[11,4.7484],[12,4.7609],[13,4.7572],[14,4.7736],[15,4.7484],[16,4.7523],[17,4.7294],[18,4.7452],[21,4.753],[22,4.7339],[23,4.7586],[24,4.7479],[25,4.7516],[26,4.7363],[27,4.7361],[28,4.6862],[29,4.7258],[31,4.7307],[32,4.7178],[33,4.7207],[35,4.7398],[39,4.7338],[40,4.7527],[42,4.735],[43,4.7787],[44,4.7381],[51,4.3995],[53,4.2812],[54,4.212],[55,4.1344],[56,3.988],[57,3.743],[58,3.4436],[59,3.1338],[60,2.8579],[61,2.66],[62,2.5842],[68,2.5845],[70,2.5624],[71,2.569],[72,2.5833],[74,2.6236],[75,2.6362],[76,2.6378],[77,2.7052],[78,2.6871],[79,2.6997],[81,2.7574],[82,2.7738],[83,2.7773],[84,2.9851],[85,3.069],[86,3.1042],[87,3.1067],[88,3.1221],[90,3.1338],[96,3.1338],[97,3.1555],[98,3.1464],[99,3.1245],[100,3.0506],[101,3.0385],[102,3.0127],[103,3.0662],[104,3.0538],[105,3.0206],[106,3.0523],[107,3.0496],[108,3.0195],[109,3.0317],[110,3.0203],[111,3.0285],[112,3.1426],[113,3.0862],[114,3.0107],[115,2.8892],[116,2.9047],[117,2.9052],[118,2.8917],[119,2.8853],[122,2.8999],[123,2.8525],[124,2.8442],[125,2.8182],[126,2.8512],[127,2.8516],[128,2.8435],[129,2.8434],[130,2.8627],[131,2.8434],[132,2.845],[133,2.8596],[134,2.8207],[135,2.8174],[137,2.8425],[138,2.8296],[139,2.8047],[140,2.8257],[141,2.8221],[142,2.8294],[144,2.7742],[145,2.7641],[146,2.7939],[147,2.7862],[148,2.7987],[149,2.7986],[150,2.8143],[151,2.8102],[152,2.8194],[153,2.7992],[154,2.7848],[155,2.7904],[156,2.8027],[157,2.8004],[158,2.7899],[159,2.8001],[160,2.8016],[161,2.819],[162,2.8125],[163,2.8309],[165,2.8254],[166,2.7727],[167,2.6822],[168,2.6108],[169,2.7341],[170,3.0963],[171,3.6264],[172,3.9076],[173,4.1589],[174,4.2788],[175,4.3228],[177,4.3183],[178,4.3774],[179,4.3666],[180,4.2423],[181,4.2717],[182,4.2555],[183,3.9307],[184,4.0428]]
       }
       ring_03_l_ctrl.Rotation.Z {
        type=MovieSceneScriptingFloatChannel
        default=-19.847
        thinned_from=185
        keys [[0,-27.2472],[1,-27.3024],[2,-27.2999],[3,-27.388],[4,-27.379],[6,-27.5234],[7,-27.387],[8,-27.3581],[10,-27.5907],[11,-27.3561],[12,-27.5053],[13,-27.4604],[14,-27.6558],[15,-27.3566],[16,-27.4022],[17,-27.1323],[18,-27.3187],[20,-27.3291],[21,-27.4102],[22,-27.1849],[23,-27.477],[24,-27.3509],[25,-27.3942],[26,-27.2133],[27,-27.2118],[28,-26.6289],[29,-27.0901],[31,-27.148],[32,-26.996],[33,-27.0303],[34,-27.1897],[35,-27.255],[37,-27.2426],[38,-27.1784],[39,-27.1839],[40,-27.4071],[42,-27.1996],[43,-27.7173],[46,-26.2849],[50,-24.5311],[52,-23.5249],[53,-22.9452],[54,-22.2977],[55,-21.5696],[56,-20.2176],[57,-17.9736],[58,-15.2386],[59,-12.4135],[60,-9.8992],[61,-8.0967],[62,-7.4067],[68,-7.4115],[70,-7.2399],[71,-7.288],[72,-7.3939],[74,-7.7089],[75,-7.8061],[76,-7.8169],[77,-8.3327],[78,-8.1975],[79,-8.2927],[81,-8.7351],[82,-8.8695],[83,-8.9043],[84,-10.5612],[85,-11.2388],[86,-11.526],[87,-11.547],[88,-11.6726],[89,-11.7403],[90,-11.7681],[96,-11.7683],[97,-11.9476],[98,-11.8732],[99,-11.693],[100,-11.0882],[101,-10.9893],[102,-10.7801],[103,-11.2147],[104,-11.1139],[105,-10.8441],[106,-11.1015],[107,-11.0798],[108,-10.8351],[109,-10.9339],[110,-10.8421],[111,-10.9089],[112,-11.8421],[113,-11.3778],[114,-10.7639],[115,-9.7888],[116,-9.9128],[117,-9.9159],[118,-9.8084],[119,-9.7579],[120,-9.8254],[122,-9.8742],[123,-9.497],[124,-9.4319],[125,-9.2261],[126,-9.487],[127,-9.4901],[128,-9.4259],[129,-9.4255],[130,-9.5779],[131,-9.4249],[132,-9.438],[133,-9.5533],[134,-9.2461],[135,-9.2197],[137,-9.418],[138,-9.3164],[139,-9.1199],[140,-9.2853],[141,-9.2566],[142,-9.3146],[144,-8.88],[145,-8.801],[146,-9.0347],[147,-8.9746],[148,-9.0727],[149,-9.0721],[150,-9.1955],[151,-9.163],[152,-9.2356],[153,-9.0765],[154,-8.9636],[155,-9.0076],[156,-9.1042],[157,-9.0856],[158,-9.0034],[159,-9.0835],[160,-9.0953],[161,-9.2327],[162,-9.181],[163,-9.3267],[164,-9.294],[165,-9.2827],[166,-8.8529],[167,-8.1621],[168,-7.6105],[169,-8.5786],[170,-11.4846],[171,-16.0019],[172,-18.5607],[173,-20.9915],[174,-22.2005],[175,-22.6534],[177,-22.6065],[178,-23.224],[179,-23.11],[180,-21.8261],[181,-22.1213],[182,-21.9242],[183,-18.7724],[184,-19.847]]
       }
       ring_03_l_ctrl.Scale.X {type=MovieSceneScriptingFloatChannel, default=1, const [1,185]}
       ring_03_l_ctrl.Scale.Y {type=MovieSceneScriptingFloatChannel, default=1, const [1,185]}
       ring_03_l_ctrl.Scale.Z {type=MovieSceneScriptingFloatChannel, default=1, const [1,185]}
       pinky_metacarpal_l_ctrl.Location.X {
        type=MovieSceneScriptingFloatChannel
        default=0
        thinned_from=185
        keys [[0,0],[47,0],[55,-0.0004],[61,-0.0016],[68,-0.001],[70,-0.0002],[72,-0.001],[74,0],[75,-0.0002],[76,-0.0001],[77,-0.0006],[78,-0.0001],[82,-0.0003],[84,-0.0002],[85,0],[165,0],[166,-0.0015],[167,0],[168,-0.0014],[169,-0.0011],[171,-0.0008],[172,-0.0001],[180,-0.0001],[181,-0.0009],[182,-0.0002],[183,-0.0005],[184,0]]
       }
       pinky_metacarpal_l_ctrl.Location.Y {
        type=MovieSceneScriptingFloatChannel
        default=0
        thinned_from=185
        keys [[0,0],[63,-0.0003],[68,0.0002],[74,0],[75,0.0002],[76,0.0001],[77,0.0007],[78,0.0002],[82,0.0004],[83,0.0001],[165,0],[166,-0.0006],[167,0],[168,-0.0007],[169,-0.0003],[180,0],[182,-0.0005],[183,0.0002],[184,0]]
       }
       pinky_metacarpal_l_ctrl.Location.Z {
        type=MovieSceneScriptingFloatChannel
        default=0
        thinned_from=185
        keys [[0,0],[47,0],[48,-0.0002],[55,-0.0007],[61,-0.0026],[64,-0.0025],[68,-0.0014],[70,-0.0003],[72,-0.0014],[74,0],[75,-0.0003],[76,-0.0001],[77,-0.0007],[78,-0.0001],[82,-0.0004],[84,-0.0002],[85,0],[165,0],[166,-0.0017],[167,0],[168,-0.0009],[169,-0.0006],[170,-0.0007],[171,-0.0005],[172,-0.0001],[180,-0.0001],[181,-0.0012],[182,-0.0015],[183,-0.0006],[184,0]]
       }
       pinky_metacarpal_l_ctrl.Rotation.X {type=MovieSceneScriptingFloatChannel, default=-0.5768, const [-0.5768,185]}
       pinky_metacarpal_l_ctrl.Rotation.Y {type=MovieSceneScriptingFloatChannel, default=7.1368, const [7.1368,185]}
       pinky_metacarpal_l_ctrl.Rotation.Z {type=MovieSceneScriptingFloatChannel, default=-4.6328, const [-4.6328,185]}
       pinky_metacarpal_l_ctrl.Scale.X {type=MovieSceneScriptingFloatChannel, default=1, const [1,185]}
       pinky_metacarpal_l_ctrl.Scale.Y {type=MovieSceneScriptingFloatChannel, default=1, const [1,185]}
       pinky_metacarpal_l_ctrl.Scale.Z {type=MovieSceneScriptingFloatChannel, default=1, const [1,185]}
       pinky_01_l_ctrl.Location.X {
        type=MovieSceneScriptingFloatChannel
        default=0
        thinned_from=185
        keys [[0,0],[50,-0.0003],[55,-0.001],[57,-0.0017],[60,-0.0032],[61,-0.0036],[63,-0.0037],[68,-0.0024],[70,-0.0004],[72,-0.0023],[74,0],[75,-0.0007],[76,-0.0004],[77,-0.0022],[78,-0.0005],[82,-0.0012],[83,-0.0004],[84,-0.0004],[85,0],[165,0],[166,-0.0042],[167,0],[168,-0.0039],[169,-0.0028],[171,-0.0018],[172,-0.0002],[176,0],[179,0],[180,-0.0004],[181,-0.0025],[182,-0.0006],[183,-0.0008],[184,0]]
       }
       pinky_01_l_ctrl.Location.Y {
        type=MovieSceneScriptingFloatChannel
        default=0
        thinned_from=185
        keys [[0,0],[49,-0.0003],[55,-0.0011],[57,-0.0018],[60,-0.0036],[61,-0.004],[62,-0.0041],[64,-0.0037],[68,-0.0019],[69,-0.001],[70,-0.0003],[72,-0.002],[74,0],[75,-0.0003],[76,-0.0001],[77,-0.0006],[78,-0.0001],[84,-0.0003],[85,0],[165,0],[166,-0.0024],[167,0],[168,-0.0023],[169,-0.0011],[172,0],[179,0],[180,-0.0002],[181,-0.0018],[182,-0.0022],[183,-0.0005],[184,0]]
       }
       pinky_01_l_ctrl.Location.Z {
        type=MovieSceneScriptingFloatChannel
        default=0
        thinned_from=185
        keys [[0,0],[82,0.0001],[83,-0.0002],[84,-0.0002],[85,0],[165,0],[166,0.0003],[167,0],[168,0.0003],[169,0.0002],[170,-0.0002],[171,-0.0003],[172,0],[180,0],[181,0.0004],[182,0.0006],[183,-0.0007],[184,0]]
       }
       pinky_01_l_ctrl.Rotation.X {
        type=MovieSceneScriptingFloatChannel
        default=-4.8111
        thinned_from=185
        keys [[0,-3.8974],[1,-3.9408],[2,-3.8863],[3,-3.906],[4,-3.8929],[5,-3.9131],[6,-3.8947],[7,-3.9217],[8,-3.8938],[9,-3.8919],[10,-3.8587],[11,-3.8843],[12,-3.8492],[13,-3.856],[14,-3.8862],[15,-3.8674],[16,-3.8683],[17,-3.8909],[18,-3.9037],[19,-3.9028],[20,-3.8763],[23,-3.8917],[24,-3.9115],[25,-3.8858],[26,-3.8938],[27,-3.9555],[28,-3.9636],[29,-3.9475],[30,-3.9228],[31,-3.9119],[32,-3.9383],[33,-3.9242],[34,-3.9341],[35,-3.9204],[36,-3.9184],[37,-3.8986],[38,-3.9092],[40,-3.8911],[41,-3.8928],[42,-3.9259],[43,-3.8229],[45,-3.9551],[51,-4.296],[53,-4.4353],[54,-4.5171],[55,-4.6091],[56,-4.7785],[57,-5.0579],[58,-5.3978],[59,-5.7486],[60,-6.0605],[61,-6.2841],[62,-6.3697],[66,-6.3278],[68,-6.289],[69,-6.2323],[70,-6.156],[72,-6.0517],[74,-5.9631],[76,-5.7414],[77,-5.5124],[78,-5.3513],[82,-5.2513],[83,-5.1786],[84,-5.3964],[86,-5.7349],[87,-5.7806],[88,-5.8438],[89,-5.8738],[96,-5.8902],[97,-5.6764],[98,-5.4238],[99,-5.1872],[100,-5.3212],[101,-5.3412],[102,-5.3966],[104,-5.4169],[105,-5.3986],[106,-5.4399],[107,-5.4421],[109,-5.4897],[110,-5.4688],[111,-5.4697],[112,-5.5921],[113,-5.5623],[114,-5.4643],[115,-5.3005],[116,-5.2876],[117,-5.2871],[119,-5.3148],[120,-5.3031],[122,-5.3242],[123,-5.3096],[125,-5.3227],[126,-5.3735],[127,-5.3606],[128,-5.3999],[129,-5.3805],[130,-5.3771],[131,-5.4046],[133,-5.4314],[134,-5.3867],[135,-5.4238],[136,-5.4018],[138,-5.4138],[139,-5.3909],[140,-5.4121],[142,-5.3666],[143,-5.3918],[144,-5.3861],[145,-5.3675],[146,-5.3991],[147,-5.35],[148,-5.3898],[149,-5.4433],[150,-5.41],[152,-5.4173],[153,-5.4068],[155,-5.4302],[156,-5.4098],[158,-5.4463],[160,-5.4434],[162,-5.5016],[163,-5.4892],[164,-5.5277],[166,-5.5811],[167,-5.8399],[168,-5.892],[169,-5.8189],[170,-5.543],[171,-5.0989],[173,-4.7148],[174,-4.4924],[175,-4.4565],[176,-4.3688],[177,-4.4014],[178,-4.4703],[179,-4.4923],[181,-4.4904],[182,-4.8217],[183,-4.9281],[184,-4.8111]]
       }
       pinky_01_l_ctrl.Rotation.Y {
        type=MovieSceneScriptingFloatChannel
        default=-7.1245
        thinned_from=185
        keys [[0,-9.8152],[1,-9.8204],[2,-9.742],[3,-9.8305],[4,-9.7719],[5,-9.7601],[6,-9.8837],[7,-9.8157],[8,-9.7707],[9,-9.8491],[10,-9.8811],[12,-9.7769],[13,-9.8615],[14,-9.875],[15,-9.8198],[16,-9.9057],[17,-9.9039],[18,-9.8043],[19,-9.8082],[20,-9.8439],[23,-9.7698],[25,-9.8119],[26,-9.801],[27,-9.7471],[28,-9.763],[29,-9.7101],[31,-9.7917],[32,-9.771],[33,-9.7713],[34,-9.7242],[36,-9.8463],[37,-9.7643],[38,-9.7896],[39,-9.8378],[40,-9.8123],[41,-9.8464],[42,-9.8576],[43,-9.9404],[46,-9.5684],[51,-9.0101],[53,-8.7348],[54,-8.5732],[55,-8.3913],[56,-8.056],[57,-7.5019],[58,-6.8277],[59,-6.1317],[60,-5.5125],[61,-5.0688],[62,-4.899],[63,-4.9081],[64,-4.9332],[66,-5.0186],[68,-5.1297],[69,-5.2625],[70,-5.4298],[72,-5.6363],[73,-5.788],[74,-5.963],[75,-6.1759],[76,-6.4365],[77,-6.9753],[78,-7.3549],[79,-7.4199],[81,-7.5089],[82,-7.574],[83,-8.0934],[84,-8.4112],[85,-8.2392],[86,-8.5826],[87,-8.7041],[88,-8.7625],[90,-8.8248],[97,-8.835],[98,-8.8071],[99,-8.8038],[101,-8.6712],[102,-8.5784],[103,-8.5337],[104,-8.5307],[105,-8.51],[106,-8.5351],[107,-8.4986],[108,-8.5309],[109,-8.5086],[110,-8.5408],[111,-8.5428],[112,-8.59],[113,-8.5098],[114,-8.4113],[115,-8.1989],[116,-8.1534],[117,-8.1539],[118,-8.174],[119,-8.1512],[120,-8.0818],[121,-8.1197],[123,-8.1167],[124,-8.0862],[127,-8.0287],[128,-8.0476],[129,-8.0234],[130,-8.0162],[132,-7.9034],[133,-7.8711],[134,-7.9292],[135,-7.9539],[136,-7.9164],[138,-7.9373],[139,-7.9146],[140,-7.9176],[141,-7.8957],[143,-7.9307],[145,-8.0322],[146,-8.0356],[147,-7.9146],[149,-7.9884],[150,-7.945],[151,-7.9475],[152,-7.8678],[153,-7.8359],[154,-7.9091],[155,-7.9128],[156,-7.858],[157,-7.8448],[158,-7.8595],[159,-7.8108],[160,-7.8095],[162,-7.9066],[163,-7.854],[164,-7.8813],[165,-7.8604],[166,-7.483],[167,-6.9754],[168,-6.4395],[169,-6.3815],[170,-6.6278],[171,-7.5152],[172,-7.9666],[173,-8.3485],[174,-8.4238],[175,-8.4365],[176,-8.2892],[177,-8.3303],[178,-8.3533],[179,-8.3285],[180,-8.2622],[181,-8.2769],[182,-7.5619],[183,-7.0899],[184,-7.1245]]
       }
       pinky_01_l_ctrl.Rotation.Z {
        type=MovieSceneScriptingFloatChannel
        default=-26.183
        thinned_from=185
        keys [[0,-40.5786],[1,-40.6221],[2,-40.8253],[3,-40.3095],[5,-40.3262],[6,-40.4147],[7,-40.4178],[8,-40.3349],[9,-40.5142],[10,-40.4513],[11,-40.472],[13,-40.8107],[14,-40.6819],[15,-40.7348],[16,-40.929],[17,-40.6017],[18,-40.4928],[20,-40.5476],[21,-40.4283],[22,-40.4168],[24,-40.2361],[26,-40.3634],[28,-39.906],[29,-40.2664],[31,-40.2382],[32,-40.3354],[33,-40.2568],[34,-40.3118],[35,-40.1967],[36,-40.4186],[37,-40.3802],[39,-40.5545],[40,-40.4776],[42,-40.5251],[43,-41.0991],[51,-36.9574],[53,-35.7072],[54,-34.9741],[55,-34.15],[56,-32.6175],[57,-30.0724],[58,-26.9696],[59,-23.7642],[60,-20.9112],[61,-18.8657],[62,-18.0828],[68,-18.1515],[69,-18.4341],[70,-18.8878],[72,-19.5888],[73,-19.8393],[74,-20.1402],[75,-20.9973],[76,-22.1369],[77,-24.3842],[78,-25.8417],[79,-26.1085],[81,-26.4737],[82,-26.7406],[83,-28.4294],[84,-24.977],[85,-24.5943],[87,-24.8747],[89,-24.835],[96,-24.9486],[98,-26.1905],[99,-27.4956],[100,-29.3247],[101,-29.0578],[102,-28.492],[103,-28.0403],[104,-27.7467],[105,-27.8055],[106,-27.6023],[109,-27.4837],[110,-27.2972],[111,-27.344],[112,-27.6373],[113,-27.4798],[114,-26.8354],[115,-26.6302],[117,-26.577],[118,-26.6618],[119,-26.2842],[120,-26.217],[121,-26.2458],[122,-26.1501],[123,-26.3472],[124,-26.1279],[125,-26.0916],[126,-25.9146],[127,-25.8595],[128,-25.6836],[130,-25.5425],[131,-25.6602],[133,-25.4621],[134,-25.6483],[135,-25.381],[136,-25.4402],[137,-25.4228],[138,-25.578],[139,-25.564],[140,-25.4813],[142,-25.4421],[143,-25.5053],[144,-25.5144],[145,-25.4273],[146,-25.0508],[148,-24.9779],[149,-25.0259],[150,-24.8838],[151,-24.976],[152,-24.8856],[153,-24.662],[154,-24.6237],[155,-24.4701],[156,-24.4081],[157,-24.2554],[158,-24.3078],[159,-24.2109],[160,-24.2175],[161,-24.3015],[162,-24.0435],[163,-24.1253],[164,-23.7489],[165,-23.6592],[166,-22.9988],[167,-21.5924],[168,-22.4075],[169,-24.1801],[170,-26.1519],[171,-29.4068],[172,-31.4032],[173,-32.6121],[174,-32.8802],[175,-32.8989],[176,-32.6188],[177,-32.7308],[178,-32.2603],[180,-31.8094],[181,-31.1006],[182,-26.2609],[183,-25.2229],[184,-26.183]]
       }
       pinky_01_l_ctrl.Scale.X {type=MovieSceneScriptingFloatChannel, default=1, const [1,185]}
       pinky_01_l_ctrl.Scale.Y {type=MovieSceneScriptingFloatChannel, default=1, const [1,185]}
       pinky_01_l_ctrl.Scale.Z {type=MovieSceneScriptingFloatChannel, default=1, const [1,185]}
       pinky_02_l_ctrl.Location.X {
        type=MovieSceneScriptingFloatChannel
        default=0
        thinned_from=185
        keys [[0,0],[76,0],[77,-0.0003],[82,-0.0001],[84,-0.0003],[85,0],[165,0],[166,-0.001],[167,0],[168,-0.0009],[171,-0.0004],[172,-0.0001],[182,-0.0002],[184,0]]
       }
       pinky_02_l_ctrl.Location.Y {
        type=MovieSceneScriptingFloatChannel
        default=0
        thinned_from=185
        keys [[0,0],[56,-0.0003],[62,-0.001],[68,-0.0008],[70,-0.0001],[72,-0.0007],[74,0],[75,-0.0002],[76,-0.0001],[77,-0.0006],[78,-0.0001],[82,-0.0004],[84,-0.0002],[85,0],[165,0],[166,-0.0015],[167,0],[168,-0.001],[169,-0.0006],[172,0],[180,-0.0001],[181,-0.0005],[182,0.0002],[184,0]]
       }
       pinky_02_l_ctrl.Location.Z {
        type=MovieSceneScriptingFloatChannel
        default=0
        thinned_from=185
        keys [[0,0],[68,0.0002],[82,0],[83,-0.0003],[84,-0.0003],[85,0],[165,0],[166,0.0009],[167,0],[168,0.0006],[172,0],[180,0.0001],[181,0.0005],[182,0.0001],[183,-0.0001],[184,0]]
       }
       pinky_02_l_ctrl.Rotation.X {
        type=MovieSceneScriptingFloatChannel
        default=-10.7462
        thinned_from=185
        keys [[0,-11.8204],[1,-11.8365],[2,-11.8216],[3,-11.8357],[8,-11.8151],[14,-11.8246],[16,-11.803],[22,-11.8121],[27,-11.7875],[28,-11.8132],[29,-11.7864],[33,-11.7958],[35,-11.7839],[38,-11.8137],[39,-11.7857],[42,-11.7876],[43,-11.7475],[44,-11.7752],[45,-11.7619],[47,-11.6633],[49,-11.485],[52,-11.1118],[55,-10.6715],[56,-10.3885],[57,-9.9009],[58,-9.299],[59,-8.6734],[60,-8.1145],[61,-7.7129],[62,-7.5589],[63,-7.5684],[69,-7.7711],[72,-7.8199],[74,-7.9343],[75,-7.9337],[76,-8.026],[77,-8.2482],[78,-8.3373],[82,-8.499],[83,-8.5844],[84,-9.0689],[85,-9.2724],[86,-9.3575],[89,-9.4471],[96,-9.4594],[98,-9.3281],[99,-9.2255],[100,-9.1519],[101,-9.0984],[103,-9.083],[104,-9.1536],[105,-9.0921],[106,-9.1418],[107,-9.1399],[108,-9.1176],[110,-9.11],[111,-9.1262],[112,-9.2642],[113,-9.1442],[114,-9.1019],[115,-8.8942],[117,-8.8995],[118,-8.8407],[119,-8.8586],[120,-8.8455],[121,-8.8648],[122,-8.8487],[123,-8.7874],[124,-8.7725],[125,-8.7848],[126,-8.8512],[127,-8.8296],[128,-8.8765],[129,-8.8484],[130,-8.8593],[131,-8.8321],[132,-8.8271],[133,-8.8471],[134,-8.7558],[136,-8.8482],[137,-8.8095],[138,-8.8128],[139,-8.7795],[140,-8.7999],[141,-8.8007],[142,-8.7641],[143,-8.7715],[144,-8.7471],[145,-8.7469],[146,-8.7839],[147,-8.6966],[148,-8.7603],[149,-8.7971],[151,-8.7773],[152,-8.7996],[153,-8.7465],[155,-8.781],[156,-8.8226],[157,-8.8209],[159,-8.7459],[160,-8.7501],[162,-8.8165],[165,-8.8504],[167,-8.4773],[168,-8.5382],[169,-8.8368],[170,-9.4042],[171,-10.3645],[172,-10.7826],[173,-11.1347],[175,-11.2389],[176,-11.2397],[179,-11.3568],[180,-11.2167],[181,-11.2327],[182,-11.1217],[183,-10.6373],[184,-10.7462]]
       }
       pinky_02_l_ctrl.Rotation.Y {
        type=MovieSceneScriptingFloatChannel
        default=-5.3889
        thinned_from=185
        keys [[0,-7.7792],[1,-7.8305],[2,-7.7829],[3,-7.8279],[4,-7.7827],[5,-7.8054],[7,-7.7945],[8,-7.7623],[9,-7.7968],[11,-7.788],[12,-7.7635],[14,-7.7924],[16,-7.7243],[17,-7.7644],[19,-7.7333],[21,-7.7235],[22,-7.7527],[23,-7.6957],[24,-7.7021],[25,-7.6509],[27,-7.6762],[28,-7.7562],[29,-7.6731],[30,-7.6846],[31,-7.7178],[33,-7.7019],[34,-7.6617],[35,-7.6654],[36,-7.713],[37,-7.7161],[38,-7.7576],[39,-7.6709],[40,-7.6934],[41,-7.6677],[42,-7.6768],[43,-7.5562],[44,-7.6389],[45,-7.621],[46,-7.5696],[47,-7.488],[48,-7.3795],[49,-7.2476],[50,-7.0956],[52,-6.7445],[55,-6.1508],[56,-5.7692],[57,-5.1117],[58,-4.3002],[59,-3.4567],[60,-2.7031],[61,-2.1616],[62,-1.9539],[63,-1.9609],[64,-1.9791],[67,-2.0632],[69,-2.1103],[70,-2.1275],[72,-2.1438],[73,-2.1915],[74,-2.2331],[75,-2.2317],[76,-2.3036],[77,-2.4797],[78,-2.5562],[79,-2.594],[81,-2.6568],[82,-2.6942],[83,-2.7687],[84,-3.2331],[85,-3.4438],[86,-3.5337],[88,-3.6045],[89,-3.6302],[90,-3.6417],[96,-3.6437],[98,-3.4998],[99,-3.3908],[100,-3.3147],[101,-3.2601],[102,-3.2572],[103,-3.2446],[104,-3.3166],[105,-3.2539],[106,-3.3045],[107,-3.3025],[108,-3.2797],[110,-3.2719],[111,-3.2886],[112,-3.4323],[113,-3.3069],[114,-3.2638],[115,-3.0584],[116,-3.0573],[117,-3.0636],[118,-3.0072],[119,-3.0243],[120,-3.0118],[121,-3.0302],[122,-3.0149],[123,-2.9569],[124,-2.9429],[125,-2.9545],[126,-3.0173],[127,-2.9967],[128,-3.0414],[129,-3.0146],[130,-3.025],[131,-2.9991],[132,-2.9943],[133,-3.0134],[134,-2.9275],[136,-3.0143],[137,-2.9777],[138,-2.9808],[139,-2.9495],[140,-2.9687],[141,-2.9694],[142,-2.9351],[143,-2.942],[144,-2.9193],[145,-2.9191],[146,-2.9536],[147,-2.8727],[148,-2.9316],[149,-2.966],[150,-2.9602],[151,-2.9475],[152,-2.9684],[153,-2.9187],[155,-2.9509],[156,-2.9901],[157,-2.9885],[159,-2.9182],[160,-2.9221],[161,-2.9599],[162,-2.9843],[163,-2.9971],[164,-2.9979],[165,-3.0165],[166,-2.8331],[167,-2.6767],[168,-2.7315],[169,-3.0075],[170,-3.588],[171,-4.7959],[173,-6.0954],[174,-6.1904],[175,-6.3063],[176,-6.308],[178,-6.4636],[179,-6.5605],[180,-6.2613],[181,-6.2944],[182,-6.0825],[183,-5.2104],[184,-5.3889]]
       }
       pinky_02_l_ctrl.Rotation.Z {
        type=MovieSceneScriptingFloatChannel
        default=-24.5011
        thinned_from=185
        keys [[0,-37.924],[1,-38.1988],[2,-37.944],[3,-38.1852],[4,-37.943],[5,-38.0645],[7,-38.006],[8,-37.8332],[9,-38.0183],[11,-37.9709],[12,-37.8396],[14,-37.9948],[16,-37.6288],[17,-37.8447],[19,-37.6775],[21,-37.6246],[22,-37.7815],[23,-37.4753],[24,-37.5098],[25,-37.2345],[27,-37.3709],[28,-37.8007],[29,-37.3542],[30,-37.4159],[31,-37.594],[33,-37.5088],[34,-37.2926],[35,-37.3127],[36,-37.5683],[37,-37.5854],[38,-37.8083],[39,-37.3419],[40,-37.4629],[41,-37.3248],[42,-37.3739],[43,-36.7235],[44,-37.1699],[45,-37.054],[46,-36.721],[47,-36.1924],[48,-35.4901],[49,-34.6357],[50,-33.651],[51,-32.5577],[52,-31.3776],[54,-28.8434],[55,-27.5329],[56,-25.0619],[57,-20.8038],[58,-15.5485],[59,-10.0859],[60,-5.2061],[61,-1.6989],[62,-0.3543],[63,-0.4145],[64,-0.5729],[65,-0.7958],[66,-1.05],[67,-1.3017],[68,-1.5176],[69,-1.6996],[70,-1.8488],[72,-2.0105],[73,-2.409],[74,-2.7468],[75,-2.7448],[76,-3.343],[77,-4.8107],[78,-5.3987],[79,-5.6964],[81,-6.1975],[82,-6.4931],[83,-7.0669],[84,-10.5041],[85,-12.0013],[86,-12.635],[87,-12.8598],[88,-13.1291],[89,-13.3083],[90,-13.388],[96,-13.4018],[98,-12.4046],[99,-11.6381],[100,-11.0973],[101,-10.7061],[102,-10.6851],[103,-10.5947],[104,-11.1109],[105,-10.661],[106,-11.0242],[107,-11.0101],[108,-10.8465],[109,-10.8385],[110,-10.791],[111,-10.9101],[112,-11.9295],[113,-11.0412],[114,-10.7322],[115,-9.2372],[116,-9.2291],[117,-9.2752],[118,-8.858],[119,-8.9847],[120,-8.8924],[121,-9.0286],[122,-8.9148],[123,-8.4828],[124,-8.378],[125,-8.4647],[126,-8.9325],[127,-8.7798],[128,-9.1114],[129,-8.9124],[130,-8.99],[131,-8.7976],[132,-8.7621],[133,-8.9035],[134,-8.2616],[135,-8.601],[136,-8.9106],[137,-8.6381],[138,-8.6613],[139,-8.427],[140,-8.5705],[141,-8.5764],[142,-8.3194],[143,-8.3711],[144,-8.2004],[145,-8.199],[146,-8.4581],[147,-7.8483],[148,-8.2928],[149,-8.5507],[150,-8.5075],[151,-8.412],[152,-8.5686],[153,-8.1961],[154,-8.2961],[155,-8.4379],[156,-8.7306],[157,-8.7186],[159,-8.1923],[160,-8.2215],[161,-8.5054],[162,-8.6873],[163,-8.7831],[164,-8.7884],[165,-8.9269],[166,-7.5867],[167,-6.3414],[168,-6.7584],[169,-8.8347],[170,-12.9822],[171,-20.8788],[172,-24.8653],[173,-28.6151],[174,-29.163],[175,-29.8216],[176,-29.8316],[177,-30.3021],[178,-30.7103],[179,-31.2535],[180,-29.564],[181,-29.7522],[182,-28.5376],[183,-23.4322],[184,-24.5011]]
       }
       pinky_02_l_ctrl.Scale.X {type=MovieSceneScriptingFloatChannel, default=1, const [1,185]}
       pinky_02_l_ctrl.Scale.Y {type=MovieSceneScriptingFloatChannel, default=1, const [1,185]}
       pinky_02_l_ctrl.Scale.Z {type=MovieSceneScriptingFloatChannel, default=1, const [1,185]}
       pinky_03_l_ctrl.Location.X {
        type=MovieSceneScriptingFloatChannel
        default=0
        thinned_from=185
        keys [[0,0],[82,0],[84,-0.0001],[165,0],[166,-0.0013],[167,0],[168,-0.0008],[170,-0.0001],[180,-0.0001],[182,-0.0013],[183,-0.0002],[184,0]]
       }
       pinky_03_l_ctrl.Location.Y {
        type=MovieSceneScriptingFloatChannel
        default=0
        thinned_from=185
        keys [[0,0],[76,0],[77,-0.0002],[165,0],[166,-0.0003],[167,0],[168,-0.0002],[180,0],[181,0.0002],[182,0],[184,0]]
       }
       pinky_03_l_ctrl.Location.Z {
        type=MovieSceneScriptingFloatChannel
        default=0
        thinned_from=185
        keys [[0,0],[68,0.0002],[70,0],[72,0.0002],[74,0],[77,0.0002],[84,-0.0001],[165,0],[166,0.0011],[167,0],[168,0.0006],[171,0],[180,0],[181,0.0002],[182,-0.0002],[184,0]]
       }
       pinky_03_l_ctrl.Rotation.X {
        type=MovieSceneScriptingFloatChannel
        default=-7.2749
        thinned_from=185
        keys [[0,-8.3394],[3,-8.3608],[4,-8.3389],[5,-8.3593],[8,-8.3385],[10,-8.3649],[12,-8.352],[14,-8.3667],[15,-8.3256],[22,-8.3277],[26,-8.2762],[27,-8.2784],[28,-8.3107],[29,-8.2902],[30,-8.2885],[31,-8.3204],[34,-8.2905],[35,-8.3003],[36,-8.3343],[37,-8.3341],[38,-8.3527],[39,-8.3064],[40,-8.3354],[41,-8.3123],[42,-8.3081],[43,-8.282],[45,-8.2563],[46,-8.2173],[48,-8.0879],[50,-7.9018],[53,-7.5478],[55,-7.2826],[56,-7.0293],[57,-6.5937],[58,-6.0565],[59,-5.4984],[60,-5],[61,-4.6418],[62,-4.5045],[63,-4.5126],[65,-4.5638],[69,-4.6871],[70,-4.7043],[72,-4.7126],[74,-4.8067],[75,-4.8015],[76,-4.8619],[77,-5.078],[78,-5.085],[79,-5.1076],[82,-5.2356],[83,-5.268],[84,-5.5145],[85,-5.5966],[87,-5.5898],[90,-5.6139],[96,-5.6134],[97,-5.6347],[98,-5.6237],[99,-5.5951],[100,-5.5067],[102,-5.495],[104,-5.5672],[105,-5.5015],[106,-5.5445],[107,-5.5323],[108,-5.5048],[109,-5.4949],[112,-5.5095],[113,-5.4254],[114,-5.471],[115,-5.4473],[117,-5.4752],[118,-5.434],[122,-5.4517],[123,-5.379],[124,-5.3579],[125,-5.3719],[126,-5.4192],[127,-5.4051],[128,-5.4312],[129,-5.4213],[130,-5.4309],[131,-5.4016],[133,-5.4222],[134,-5.3551],[135,-5.3756],[136,-5.4302],[137,-5.3868],[138,-5.4179],[139,-5.3829],[141,-5.3853],[142,-5.3615],[143,-5.3665],[144,-5.3328],[145,-5.3635],[146,-5.3394],[147,-5.353],[149,-5.3526],[150,-5.3784],[151,-5.3625],[152,-5.391],[153,-5.3477],[154,-5.3416],[155,-5.3505],[156,-5.3952],[157,-5.3883],[158,-5.3569],[159,-5.3444],[160,-5.3478],[161,-5.3672],[162,-5.3655],[163,-5.3884],[164,-5.3871],[165,-5.404],[166,-5.3002],[167,-5.1535],[168,-5.1997],[169,-5.3707],[170,-5.8328],[171,-6.6075],[173,-7.3666],[174,-7.4661],[175,-7.5495],[177,-7.6501],[178,-7.672],[179,-7.7214],[180,-7.5666],[181,-7.5959],[182,-7.5153],[183,-7.1582],[184,-7.2749]]
       }
       pinky_03_l_ctrl.Rotation.Y {
        type=MovieSceneScriptingFloatChannel
        default=-2.963
        thinned_from=185
        keys [[0,-3.8594],[1,-3.8736],[3,-3.8801],[4,-3.8589],[5,-3.8787],[8,-3.8585],[10,-3.8842],[12,-3.8716],[13,-3.886],[14,-3.8859],[15,-3.846],[20,-3.8589],[22,-3.848],[23,-3.8346],[24,-3.833],[26,-3.7987],[27,-3.8007],[28,-3.8317],[29,-3.812],[30,-3.8105],[31,-3.841],[32,-3.8259],[33,-3.8264],[34,-3.8123],[35,-3.8218],[36,-3.8544],[37,-3.8543],[38,-3.8723],[39,-3.8275],[40,-3.8555],[41,-3.8333],[42,-3.8292],[43,-3.8043],[44,-3.7984],[45,-3.784],[46,-3.7588],[47,-3.7239],[48,-3.68],[50,-3.5694],[52,-3.434],[55,-3.2009],[56,-3.0489],[57,-2.788],[58,-2.4664],[59,-2.1324],[60,-1.8341],[61,-1.6198],[62,-1.5377],[63,-1.5405],[65,-1.5582],[69,-1.6021],[70,-1.6081],[72,-1.609],[74,-1.6451],[75,-1.6424],[76,-1.6648],[77,-1.7449],[78,-1.7507],[79,-1.7623],[83,-1.8258],[84,-1.9341],[85,-1.9726],[87,-1.9694],[89,-1.979],[96,-1.9804],[97,-1.9904],[98,-1.9852],[99,-1.9719],[100,-1.9312],[102,-1.9259],[104,-1.9589],[105,-1.9289],[106,-1.9484],[107,-1.9429],[108,-1.9303],[109,-1.9259],[112,-1.9325],[113,-1.8948],[114,-1.9151],[115,-1.9045],[117,-1.917],[118,-1.8986],[122,-1.9065],[123,-1.8743],[124,-1.8652],[125,-1.8713],[126,-1.892],[127,-1.8858],[128,-1.8974],[129,-1.8929],[130,-1.8972],[131,-1.8843],[132,-1.8911],[133,-1.8933],[134,-1.8639],[135,-1.8729],[136,-1.8969],[137,-1.8778],[138,-1.8914],[139,-1.8761],[141,-1.8771],[142,-1.8667],[143,-1.8689],[144,-1.8543],[145,-1.8676],[146,-1.8572],[147,-1.863],[148,-1.8611],[149,-1.8629],[150,-1.8741],[151,-1.8672],[152,-1.8796],[153,-1.8608],[154,-1.8581],[155,-1.862],[156,-1.8814],[157,-1.8784],[158,-1.8648],[159,-1.8593],[160,-1.8608],[161,-1.8692],[162,-1.8685],[163,-1.8785],[164,-1.8779],[165,-1.8853],[166,-1.8364],[167,-1.7791],[168,-1.7988],[169,-1.8728],[170,-2.09],[171,-2.5184],[172,-2.7587],[173,-3.0317],[174,-3.1062],[175,-3.1709],[177,-3.2507],[178,-3.2684],[179,-3.3087],[180,-3.1846],[181,-3.2077],[182,-3.1449],[183,-2.8792],[184,-2.963]]
       }
       pinky_03_l_ctrl.Rotation.Z {
        type=MovieSceneScriptingFloatChannel
        default=-25.4603
        thinned_from=185
        keys [[0,-33.4227],[1,-33.5395],[3,-33.5932],[4,-33.4194],[5,-33.5816],[8,-33.416],[10,-33.6263],[12,-33.5232],[13,-33.6412],[14,-33.6407],[15,-33.313],[20,-33.4188],[22,-33.3298],[23,-33.2189],[24,-33.2054],[26,-32.9223],[27,-32.9394],[28,-33.1954],[29,-33.0327],[30,-33.0195],[31,-33.272],[32,-33.1477],[33,-33.151],[34,-33.0351],[35,-33.1132],[36,-33.3825],[37,-33.3809],[38,-33.5286],[39,-33.1607],[40,-33.3909],[41,-33.2081],[42,-33.1749],[43,-32.9685],[44,-32.9201],[45,-32.7796],[46,-32.5109],[47,-32.1257],[48,-31.6354],[49,-31.0515],[50,-30.3854],[52,-28.8528],[54,-27.1295],[55,-26.2251],[56,-24.5185],[57,-21.5863],[58,-17.9709],[59,-14.2147],[60,-10.8602],[61,-8.4496],[62,-7.5256],[63,-7.5749],[64,-7.7046],[65,-7.8871],[67,-8.301],[69,-8.6413],[70,-8.7457],[72,-8.7904],[74,-9.371],[75,-9.3377],[76,-9.7078],[77,-11.0331],[78,-11.0798],[79,-11.2224],[80,-11.4679],[81,-11.7524],[82,-12.0122],[83,-12.2104],[84,-13.755],[85,-14.2782],[86,-14.237],[87,-14.2354],[89,-14.3665],[90,-14.3878],[96,-14.3845],[97,-14.5198],[98,-14.4492],[99,-14.2685],[100,-13.7108],[101,-13.6893],[102,-13.6371],[104,-14.0927],[105,-13.6782],[106,-13.949],[107,-13.8723],[108,-13.6988],[109,-13.6367],[110,-13.6861],[112,-13.7289],[113,-13.1997],[114,-13.4861],[115,-13.3374],[116,-13.4479],[117,-13.5126],[118,-13.2535],[120,-13.2926],[122,-13.365],[123,-12.9087],[124,-12.7772],[125,-12.8645],[126,-13.1608],[127,-13.0728],[128,-13.2363],[129,-13.1738],[130,-13.2342],[131,-13.0505],[132,-13.1482],[133,-13.1796],[134,-12.7594],[135,-12.8877],[136,-13.2292],[137,-12.9576],[138,-13.1525],[139,-12.9333],[140,-12.9607],[141,-12.9484],[142,-12.7991],[143,-12.8307],[144,-12.6202],[145,-12.812],[146,-12.6614],[147,-12.7461],[148,-12.7179],[149,-12.744],[150,-12.905],[151,-12.806],[152,-12.984],[153,-12.7131],[154,-12.6754],[155,-12.7307],[156,-13.0104],[157,-12.967],[158,-12.771],[159,-12.6924],[160,-12.714],[161,-12.8351],[162,-12.8244],[163,-12.9676],[164,-12.9597],[165,-13.0656],[166,-12.4028],[167,-11.5048],[168,-11.7941],[169,-12.8643],[170,-15.7927],[171,-20.868],[172,-23.422],[173,-26.1157],[174,-26.8263],[175,-27.4297],[177,-28.1643],[178,-28.3254],[179,-28.6899],[180,-27.5563],[181,-27.7698],[182,-27.2038],[183,-24.6401],[184,-25.4603]]
       }
       pinky_03_l_ctrl.Scale.X {type=MovieSceneScriptingFloatChannel, default=1, const [1,185]}
       pinky_03_l_ctrl.Scale.Y {type=MovieSceneScriptingFloatChannel, default=1, const [1,185]}
       pinky_03_l_ctrl.Scale.Z {type=MovieSceneScriptingFloatChannel, default=1, const [1,185]}
       hand_l_ik_ctrl.Location.X {
        type=MovieSceneScriptingFloatChannel
        default=-30.623
        thinned_from=185
        keys [[0,19.9554],[2,20.0582],[12,20.9332],[14,21.0011],[16,21.011],[17,20.9711],[19,20.7918],[20,20.74],[23,20.4195],[25,20.3148],[26,20.3166],[27,20.3826],[32,20.9632],[33,21.043],[40,21.1682],[42,21.1587],[44,21.1844],[45,21.51],[46,22.2127],[47,23.1354],[48,24.122],[50,26.691],[51,28.434],[52,28.5136],[53,25.9183],[54,24.4085],[55,20.1526],[56,18.3288],[57,16.6824],[58,15.6178],[59,13.9824],[60,12.1677],[61,9.8235],[62,8.075],[63,5.9245],[64,4.1779],[65,2.9992],[66,1.6491],[67,0.4747],[68,-0.8114],[69,-2.36],[70,-4.0841],[71,-5.8949],[72,-7.9227],[73,-9.8752],[75,-13.4621],[76,-15.0378],[77,-17.1957],[78,-19.2251],[79,-19.8889],[80,-20.4357],[82,-21.4275],[83,-22.6251],[84,-22.7884],[85,-22.7945],[86,-22.7426],[87,-22.7638],[89,-22.9167],[95,-23.0641],[96,-23.0582],[99,-22.1081],[100,-21.5492],[101,-21.2658],[102,-21.1561],[103,-21.1476],[106,-21.3211],[107,-21.342],[110,-21.2741],[111,-21.2808],[112,-21.3381],[114,-21.1863],[115,-20.9801],[116,-20.9227],[122,-20.9618],[129,-20.7862],[133,-20.831],[135,-20.7062],[136,-20.61],[137,-20.569],[139,-20.6391],[140,-20.6375],[143,-20.4181],[145,-20.4156],[146,-20.4933],[148,-20.5042],[150,-20.3686],[154,-20.0414],[155,-20.0079],[161,-20.0167],[164,-19.8834],[165,-19.7891],[166,-19.426],[167,-17.9035],[168,-11.1055],[169,-3.095],[170,4.8157],[171,11.8494],[172,16.1663],[173,19.6491],[174,21.8349],[175,23.2924],[176,24.1132],[177,24.535],[178,24.6002],[179,24.1706],[180,23.4661],[181,21.1372],[182,-3.6103],[183,-21.7429],[184,-30.623]]
       }
       hand_l_ik_ctrl.Location.Y {
        type=MovieSceneScriptingFloatChannel
        default=-15.164
        thinned_from=185
        keys [[0,-7.0942],[1,-7.062],[2,-7.0058],[3,-6.9034],[4,-6.7439],[5,-6.5506],[6,-6.3394],[7,-6.1772],[8,-5.997],[9,-5.8296],[10,-5.6006],[11,-5.4163],[12,-5.2449],[13,-5.1732],[14,-5.2183],[15,-5.3397],[16,-5.5368],[17,-5.7718],[18,-6.1218],[19,-6.423],[20,-6.5826],[21,-6.8791],[22,-7.0934],[23,-7.2501],[24,-7.3668],[25,-7.3988],[26,-7.3446],[27,-7.149],[28,-6.8113],[30,-6.2778],[31,-5.9497],[32,-5.6491],[33,-5.4644],[34,-5.4255],[36,-5.3997],[37,-5.4261],[38,-5.5621],[39,-5.7304],[40,-5.9399],[41,-6.1627],[42,-6.3994],[43,-6.6041],[44,-6.627],[45,-6.4741],[46,-6.1244],[47,-5.662],[48,-4.9928],[49,-3.8336],[50,-2.4104],[51,-0.4508],[52,1.192],[53,1.7369],[54,2.8532],[55,2.8843],[56,4.831],[57,7.7334],[58,10.749],[59,12.6167],[60,13.5856],[62,14.0612],[63,14.0006],[64,14.0312],[65,14.1249],[66,14.2828],[67,14.5452],[68,14.8806],[70,15.9808],[71,16.5047],[73,17.4268],[75,18.2242],[76,18.4787],[77,18.6951],[78,18.7083],[82,18.3012],[83,17.85],[84,17.4642],[86,17.3549],[87,17.3457],[92,17.447],[96,17.3841],[97,17.4748],[98,17.691],[99,17.9613],[100,18.4529],[101,18.4948],[102,18.4543],[104,18.1387],[105,18.0261],[106,17.7616],[107,17.6564],[108,17.4799],[114,16.9664],[115,16.8601],[119,16.6398],[131,16.4387],[137,16.3521],[142,16.0631],[143,16.0623],[145,16.1411],[148,16.1281],[153,15.9309],[161,16.1202],[165,16.3917],[166,16.2483],[167,15.4519],[168,11.7051],[169,5.7488],[170,-0.5932],[171,-6.5515],[172,-10.2991],[173,-12.864],[174,-13.2967],[175,-11.9939],[176,-10.1175],[177,-8.605],[178,-7],[179,-6.5281],[180,-6.6953],[181,-8.8823],[182,-12.2423],[183,-19.9727],[184,-15.164]]
       }
       hand_l_ik_ctrl.Location.Z {
        type=MovieSceneScriptingFloatChannel
        default=198.3542
        thinned_from=185
        keys [[0,68.8736],[13,69.5007],[44,68.3308],[45,68.0772],[46,67.4801],[47,66.6271],[48,65.628],[49,64.3082],[50,62.747],[51,60.557],[52,59.3043],[53,59.8911],[54,59.4155],[55,60.8592],[56,59.6604],[57,56.6909],[58,51.3739],[59,45.2272],[60,38.9827],[61,34.3369],[62,30.9782],[64,29.366],[65,28.2323],[66,27.799],[67,27.5437],[69,27.9864],[70,28.032],[71,27.8911],[73,27.4513],[74,27.095],[76,26.0771],[77,24.8069],[78,23.5918],[79,23.4807],[80,23.4885],[81,23.4133],[82,23.2266],[83,22.4309],[84,22.37],[85,22.4234],[87,22.311],[91,22.2784],[92,22.237],[96,22.3653],[99,22.358],[102,22.0971],[103,22.0461],[110,22.0056],[111,21.9705],[112,21.8564],[113,21.9762],[114,22.1551],[115,22.4414],[118,22.7438],[120,22.8452],[130,22.9742],[132,23.0336],[135,22.9724],[138,22.6297],[139,22.5579],[142,22.4823],[146,22.413],[151,22.4731],[154,22.3365],[157,22.4054],[159,22.5457],[160,22.6603],[161,22.7002],[163,22.8782],[164,22.9281],[165,23.0796],[166,23.9272],[167,26.507],[168,32.2525],[169,34.824],[170,33.8712],[171,29.6945],[172,25.1842],[173,18.7803],[174,13.6876],[175,10.7534],[176,10.4042],[177,11.7757],[178,14.6912],[179,17.5253],[180,21.8456],[181,29.1585],[182,64.3327],[183,155.8957],[184,198.3542]]
       }
       hand_l_ik_ctrl.Rotation.X {
        type=MovieSceneScriptingFloatChannel
        default=217.3267
        thinned_from=185
        keys [[0,131.7736],[2,132.9012],[3,132.5174],[4,132.9082],[6,134.8916],[7,135.0385],[8,135.5398],[9,135.5915],[12,137.2709],[13,137.4083],[15,138.8476],[17,135.6917],[18,135.2536],[19,134.117],[20,133.5811],[21,133.8885],[22,133.5239],[23,131.8845],[24,131.3921],[25,131.822],[26,132.6956],[27,132.3663],[28,132.4217],[29,134.2627],[30,133.9608],[31,134.0858],[32,134.6196],[33,133.988],[36,135.5789],[37,135.5565],[38,134.4184],[39,134.1948],[40,133.1993],[41,133.0393],[42,134.2194],[43,134.9186],[44,133.5871],[45,130.9451],[47,123.0898],[48,119.9825],[49,117.3111],[50,115.5234],[51,114.0683],[52,115.0053],[53,118.2539],[54,120.1917],[55,124.0256],[56,125.6839],[57,126.5041],[58,125.3623],[59,122.9146],[60,119.1681],[61,115.7291],[62,112.9386],[63,113.2294],[64,113.8976],[65,114.0264],[67,116.6756],[70,121.8832],[73,126.8808],[74,128.0891],[75,128.7392],[76,128.8075],[78,124.451],[79,124.4667],[81,124.8016],[83,124.2312],[84,125.6951],[86,124.8995],[89,124.7003],[96,125.6437],[97,125.1686],[98,124.063],[99,122.2955],[100,117.7174],[101,116.772],[103,116.227],[106,117.3309],[109,117.7945],[110,117.6522],[111,117.9098],[113,118.8391],[116,118.601],[123,119.2092],[125,118.9144],[127,119.0497],[129,118.6904],[133,119.1019],[134,119.032],[135,118.7066],[137,117.4339],[138,117.1353],[141,117.5821],[145,117.1437],[149,117.5445],[152,117.2463],[156,116.1834],[159,115.9901],[162,116.1871],[163,116.0938],[164,115.7534],[165,115.9568],[166,116.9433],[167,119.7658],[168,125.769],[169,130.3498],[170,127.6927],[171,112.8794],[172,82.8575],[173,60.4735],[174,53.4719],[175,50.9495],[176,51.6793],[177,54.8506],[178,61.0337],[179,68.1146],[180,86.0767],[181,118.3776],[182,184.5903],[183,202.2465],[184,217.3267]]
       }
       hand_l_ik_ctrl.Rotation.Y {
        type=MovieSceneScriptingFloatChannel
        default=57.6267
        thinned_from=185
        keys [[0,81.5242],[1,81.7133],[5,81.4894],[7,81.8904],[14,82.2018],[17,81.926],[21,82.0485],[24,81.8477],[26,82.1565],[27,82.1796],[28,82.0244],[29,82.0727],[31,81.5552],[34,81.2108],[36,81.5842],[37,81.2033],[38,81.2212],[42,82.0833],[44,82.0907],[45,81.5891],[46,80.6339],[47,79.1777],[48,76.9985],[49,73.5919],[50,69.2622],[51,63.8789],[52,58.9904],[53,55.3754],[54,50.6578],[55,47.3433],[56,40.5457],[57,30.8353],[58,19.5181],[59,9.2315],[60,0.9588],[61,-4.1025],[62,-6.9547],[63,-7.8955],[64,-9.1934],[65,-10.6747],[66,-11.9319],[67,-12.8564],[69,-13.9647],[70,-14.6472],[71,-15.7846],[72,-17.2566],[73,-19.0194],[74,-20.8482],[75,-22.3433],[76,-23.1396],[77,-23.165],[78,-22.5534],[79,-22.723],[80,-23.073],[81,-23.2313],[82,-23.2321],[83,-22.1461],[84,-20.7541],[85,-19.0977],[87,-18.6711],[88,-18.5913],[90,-18.6847],[94,-19.1769],[95,-19.2591],[96,-19.2927],[97,-19.1147],[98,-19.2793],[99,-19.2311],[100,-17.6868],[101,-17.9399],[102,-18.6014],[103,-18.433],[104,-18.1777],[105,-17.9709],[106,-17.6905],[107,-17.6147],[108,-17.3032],[110,-16.9403],[111,-16.6426],[112,-15.4021],[113,-15.6782],[114,-16.2694],[115,-16.8008],[116,-16.556],[117,-16.2779],[118,-16.1421],[120,-16.096],[121,-16.1178],[122,-16.0921],[123,-15.8791],[124,-15.7505],[125,-15.4437],[126,-15.1858],[127,-15.0412],[128,-15.0447],[129,-14.9614],[131,-14.6977],[133,-14.8098],[134,-14.7378],[135,-14.5869],[136,-14.2274],[137,-13.9798],[138,-13.9054],[140,-13.5561],[141,-13.2332],[142,-13.0753],[143,-13.0037],[145,-13.2023],[146,-13.3875],[147,-13.4032],[148,-13.3401],[149,-13.4433],[150,-13.3469],[153,-12.6504],[154,-12.3977],[156,-12.3458],[157,-12.4053],[159,-12.6762],[160,-12.9134],[161,-12.8557],[163,-13.1406],[164,-13.1447],[165,-13.0336],[166,-11.9495],[167,-5.3883],[168,18.7046],[169,39.8946],[170,54.0787],[171,66.4603],[172,67.5119],[173,59.7589],[174,51.4186],[175,47.1197],[176,46.9705],[177,49.8714],[178,55.3017],[179,61.264],[180,66.5768],[181,67.323],[182,46.2118],[183,53.0197],[184,57.6267]]
       }
       hand_l_ik_ctrl.Rotation.Z {
        type=MovieSceneScriptingFloatChannel
        default=144.737
        thinned_from=185
        keys [[0,72.8112],[2,74.0459],[3,73.6667],[4,74.2192],[5,75.3818],[6,76.8047],[7,77.5188],[8,78.4334],[9,78.7801],[11,80.8013],[12,81.5669],[13,81.9815],[14,82.9668],[15,83.3199],[16,81.3665],[17,79.2228],[18,78.5054],[19,76.9901],[20,76.1984],[21,75.9673],[22,75.0753],[23,73.0025],[24,72.1562],[25,72.5134],[26,73.5573],[27,73.4618],[28,73.936],[29,76.2221],[30,76.1152],[31,76.5774],[32,77.265],[33,76.7818],[36,78.7588],[37,78.3695],[38,77.0813],[39,76.6855],[40,75.6755],[41,75.2592],[42,76.3447],[43,76.7874],[44,75.4076],[45,72.8404],[47,65.6683],[48,62.8753],[49,60.4427],[50,58.7869],[51,57.5468],[52,58.3769],[53,60.6768],[54,61.7854],[55,64.043],[56,65.3271],[57,66.9858],[58,68.8106],[59,71.4356],[60,74.902],[61,78.7764],[62,82.2919],[63,84.7771],[65,88.8333],[67,92.5622],[68,94.6743],[69,97.531],[70,100.8462],[71,104.4293],[73,111.8762],[74,115.3981],[76,123.4775],[77,130.7154],[78,137.5194],[79,139.3116],[81,142.1308],[82,143.9724],[83,149.458],[84,153.3192],[85,154.1169],[97,153.3035],[99,151.658],[101,149.4354],[102,148.7103],[103,148.7389],[104,149.4281],[106,149.7264],[109,149.1759],[110,148.7617],[112,148.5134],[113,148.8635],[117,148.1167],[127,148.9341],[134,149.0278],[137,148.1402],[141,148.2416],[143,148.0209],[148,148.5162],[154,147.8431],[161,148.3298],[165,147.3989],[166,145.4028],[167,139.4028],[168,131.6854],[169,125.4233],[170,111.2268],[171,84.9753],[172,47.8115],[173,20.5962],[174,11.7655],[175,8.624],[176,9.471],[177,13.5555],[178,20.2989],[179,27.9404],[180,46.4902],[181,77.5133],[182,122.0149],[183,125.6458],[184,144.737]]
       }
       hand_l_ik_ctrl.Scale.X {type=MovieSceneScriptingFloatChannel, default=1, const [1,185]}
       hand_l_ik_ctrl.Scale.Y {type=MovieSceneScriptingFloatChannel, default=1, const [1,185]}
       hand_l_ik_ctrl.Scale.Z {type=MovieSceneScriptingFloatChannel, default=1, const [1,185]}
       Softness {type=MovieSceneScriptingFloatChannel, default=0, const [0,185]}
       hand_l_ikorient_ctrl.Rotation.Roll {
        type=MovieSceneScriptingFloatChannel
        default=0
        thinned_from=185
        keys [[0,0],[172,0],[173,14.6393],[174,0],[175,16.3146],[176,0],[177,0],[178,17.2037],[179,0],[180,16.7825],[181,0],[184,0]]
       }
       hand_l_ikorient_ctrl.Rotation.Pitch {
        type=MovieSceneScriptingFloatChannel
        default=0
        thinned_from=185
        keys [[0,0],[172,0],[173,23.4231],[174,0],[175,23.9173],[176,0],[177,0],[178,24.2868],[179,0],[180,26.1318],[181,0],[184,0]]
       }
       hand_l_ikorient_ctrl.Rotation.Yaw {
        type=MovieSceneScriptingFloatChannel
        default=0
        thinned_from=185
        keys [[0,0],[172,0],[173,-14.7362],[174,0],[175,-11.0489],[176,0],[177,0],[178,-8.9979],[179,0],[180,-12.2491],[181,0],[184,0]]
       }
       arm_l_pv_ik_ctrl.Location.X {
        type=MovieSceneScriptingFloatChannel
        default=-0.4299
        thinned_from=185
        keys [[0,25.0497],[3,24.8698],[8,24.1319],[9,24.0134],[10,23.8226],[12,23.5312],[13,23.4672],[14,23.5372],[15,23.6659],[17,24.0712],[19,24.5674],[20,24.6873],[21,24.8973],[22,25.0495],[24,25.2246],[25,25.2296],[26,25.1779],[27,25.0353],[32,24.0696],[33,23.9505],[36,23.9085],[37,23.9703],[41,24.4421],[43,24.7965],[45,24.935],[46,25.1076],[48,25.641],[49,25.7732],[50,25.8271],[51,25.7197],[52,25.3124],[53,24.7973],[54,24.2115],[55,23.4508],[56,22.3785],[57,21.1209],[58,20.3185],[59,20.476],[60,21.359],[61,22.4117],[62,23.0694],[63,23.1497],[65,22.8749],[67,21.9651],[68,21.4203],[69,20.6041],[70,19.6899],[71,18.7224],[72,17.6139],[74,15.0335],[75,13.5922],[76,12.3331],[77,10.757],[78,9.4571],[79,9.1465],[80,8.8004],[83,8.0218],[84,8.577],[85,8.8059],[86,8.8022],[87,8.7695],[89,8.6499],[90,8.6403],[93,8.5553],[94,8.5493],[96,8.5925],[97,8.4891],[99,8.0369],[100,7.6615],[101,7.6635],[102,7.7357],[103,8.0004],[104,8.4322],[105,8.657],[106,9.1885],[107,9.379],[108,9.7332],[109,9.9532],[111,10.2579],[112,10.3551],[114,10.9843],[115,11.4244],[116,11.5924],[117,11.7336],[118,11.8156],[119,11.86],[121,11.8618],[123,11.9284],[124,11.9882],[126,12.159],[129,12.2991],[130,12.3717],[132,12.4691],[133,12.4953],[136,12.7381],[138,12.8487],[139,12.9488],[142,13.4235],[143,13.4749],[146,13.3094],[148,13.3531],[149,13.4675],[151,13.7791],[153,13.9955],[154,14.0136],[155,13.9938],[161,13.6154],[163,13.4226],[164,13.4327],[165,13.5645],[166,14.6749],[167,18.2357],[168,25.8904],[169,30.7691],[170,33.5019],[171,33.9133],[172,32.8487],[173,30.4528],[174,28.2667],[175,26.568],[176,25.7892],[177,26.0607],[178,26.7907],[179,28.1662],[180,31.3041],[181,36.4369],[182,37.2286],[183,20.3085],[184,-0.4299]]
       }
       arm_l_pv_ik_ctrl.Location.Y {
        type=MovieSceneScriptingFloatChannel
        default=186.712
        thinned_from=185
        keys [[0,86.9642],[13,88.4925],[25,87.0308],[36,87.5384],[45,86.7268],[47,85.862],[50,84.0053],[51,83.2427],[52,82.8442],[53,82.9522],[54,82.8426],[55,83.1501],[56,83.0623],[57,82.4713],[58,80.6904],[59,77.6048],[60,73.6438],[61,69.8586],[62,67.1855],[63,66.331],[65,65.5718],[67,66.0613],[68,66.5383],[75,70.9709],[76,71.3724],[77,71.4163],[78,71.217],[82,71.4882],[83,71.0581],[85,70.667],[96,70.7341],[98,71.096],[102,71.0453],[103,70.8957],[104,70.5675],[106,70.1241],[115,68.8253],[135,68.1144],[142,67.1702],[155,66.7065],[161,67.2458],[163,67.5584],[165,67.4457],[166,66.6925],[167,63.772],[168,54.3839],[169,45.1451],[171,28.9371],[172,23.9411],[173,18.4712],[174,15.4167],[175,14.3918],[176,14.8226],[177,16.0877],[178,18.3546],[179,20.3329],[180,24.4499],[181,31.998],[182,63.875],[183,144.7815],[184,186.712]]
       }
       arm_l_pv_ik_ctrl.Location.Z {
        type=MovieSceneScriptingFloatChannel
        default=97.4315
        thinned_from=185
        keys [[0,1.0589],[1,1.0285],[2,0.9754],[3,0.9351],[4,0.8552],[5,0.6528],[6,0.4717],[7,0.1871],[8,-0.088],[9,-0.2513],[10,-0.5382],[11,-0.717],[12,-0.9549],[13,-1.1041],[14,-1.2675],[15,-1.2107],[16,-1.0171],[17,-0.8001],[18,-0.5652],[19,-0.2813],[20,-0.1288],[21,0.2208],[22,0.5408],[23,0.7649],[24,0.9687],[25,1.0394],[26,1.0081],[27,0.9088],[28,0.5841],[29,0.2626],[30,0.0556],[31,-0.1871],[32,-0.4599],[33,-0.5796],[34,-0.612],[35,-0.7398],[36,-0.9281],[37,-0.8355],[38,-0.7014],[39,-0.7035],[40,-0.6461],[41,-0.5274],[42,-0.5539],[43,-0.5038],[44,-0.518],[45,-0.8533],[46,-1.634],[47,-2.6078],[48,-3.5306],[49,-4.5456],[50,-5.3908],[51,-6.678],[52,-6.2652],[53,-2.8397],[54,-0.5492],[55,4.7328],[56,6.8761],[57,7.927],[58,7.1882],[59,6.1297],[60,4.7342],[61,4.1937],[62,3.6266],[63,4.7709],[64,5.7868],[65,6.2257],[66,7.1431],[67,7.774],[68,8.5235],[69,9.197],[70,9.8047],[71,10.2767],[72,10.7847],[73,11.1936],[74,11.5368],[75,11.7478],[76,11.9102],[77,11.4243],[78,10.7628],[79,10.8972],[80,11.0892],[81,11.1217],[82,11.0204],[83,10.171],[84,9.8554],[85,9.7857],[86,10.044],[87,10.2443],[88,10.4],[89,10.5318],[91,10.7342],[92,10.8171],[93,10.9244],[95,11.0653],[96,11.0951],[97,11.3012],[98,11.3814],[99,11.249],[100,10.8235],[102,10.5676],[103,10.3079],[104,9.8063],[105,9.5604],[106,9.5418],[107,9.4291],[109,9.0163],[110,8.8444],[112,8.4473],[113,8.2289],[114,8.0845],[115,7.9597],[116,7.9051],[119,7.6589],[121,7.5196],[124,7.0781],[126,6.7999],[127,6.7042],[128,6.5866],[129,6.5291],[130,6.4476],[131,6.4195],[133,6.4069],[135,6.2831],[136,6.1995],[137,6.1693],[138,6.0981],[139,5.9909],[140,5.9185],[141,5.866],[144,5.6416],[145,5.5899],[147,5.5637],[149,5.566],[150,5.5587],[152,5.3881],[153,5.2092],[154,4.9773],[155,4.8307],[156,4.7308],[157,4.68],[158,4.6737],[159,4.6822],[161,4.7713],[162,4.8649],[163,4.9337],[164,4.8842],[165,4.7163],[166,4.196],[167,2.6124],[168,-0.701],[169,-2.7955],[170,-3.9291],[171,-4.597],[172,-5.1994],[173,-6.0086],[174,-7.0537],[175,-8.0658],[176,-8.6226],[177,-9.3967],[178,-10.2985],[179,-10.6938],[180,-10.7167],[181,-9.0623],[182,16.793],[183,73.921],[184,97.4315]]
       }
       arm_l_pv_ik_ctrl.Rotation.X {
        type=MovieSceneScriptingFloatChannel
        default=-83.0591
        thinned_from=185
        keys [[0,-31.6425],[1,-31.7387],[2,-31.7152],[4,-31.2635],[5,-31.746],[6,-31.9637],[7,-32.9628],[8,-33.7696],[9,-34.0678],[10,-34.817],[11,-35.1364],[12,-35.745],[14,-36.8396],[15,-36.6469],[16,-35.7961],[17,-35.1489],[18,-34.7925],[19,-34.1936],[20,-33.8696],[21,-33.0706],[22,-32.1434],[23,-31.5308],[24,-30.8535],[25,-30.709],[26,-30.7275],[27,-30.5417],[28,-31.0031],[29,-31.5798],[31,-31.666],[32,-32.0173],[33,-31.9134],[34,-31.9614],[35,-32.4796],[36,-33.2486],[38,-32.1114],[39,-32.2793],[40,-32.2471],[41,-32.0457],[42,-32.4779],[43,-32.3103],[44,-32.3788],[45,-31.7987],[46,-30.7334],[47,-29.0166],[48,-26.3281],[49,-22.0105],[50,-16.8971],[51,-11.3904],[52,-6.9794],[53,-3.3632],[54,0.8253],[55,4.3275],[56,7.8049],[57,10.1664],[58,11.1149],[59,11.0877],[60,10.7645],[61,10.5274],[62,10.4774],[63,10.6136],[64,11.032],[66,11.936],[67,12.1276],[68,12.1266],[69,11.8668],[70,11.469],[71,10.9101],[72,10.0917],[73,9.1228],[74,8.2761],[75,7.3068],[76,6.2921],[77,4.5164],[78,2.6625],[79,2.124],[80,1.6163],[82,0.4654],[83,-1.0254],[84,-1.3975],[85,-1.6132],[86,-1.2464],[87,-1.0429],[89,-0.9907],[90,-0.944],[91,-0.8926],[92,-0.8026],[93,-0.7808],[94,-0.7511],[95,-0.753],[96,-0.7898],[97,-0.2363],[98,0.0638],[99,0.0718],[100,-0.0208],[101,0.1409],[102,0.0622],[103,-0.2961],[104,-1.1138],[105,-1.5073],[106,-1.5832],[107,-1.7457],[109,-2.3773],[110,-2.6705],[111,-2.9957],[112,-3.2522],[113,-3.5614],[114,-3.7779],[115,-3.9425],[116,-4.0414],[117,-4.2304],[118,-4.4543],[119,-4.6365],[120,-4.798],[121,-4.9017],[122,-5.1084],[124,-5.4265],[125,-5.5711],[126,-5.7333],[127,-5.8299],[128,-5.9607],[129,-6.0507],[130,-6.1949],[131,-6.2783],[132,-6.3073],[133,-6.2983],[135,-6.179],[136,-6.0749],[137,-6.0157],[138,-6.132],[139,-6.3189],[140,-6.4112],[141,-6.3781],[142,-6.3118],[143,-6.2731],[144,-6.306],[145,-6.3251],[146,-6.3766],[147,-6.4143],[148,-6.3966],[149,-6.3388],[150,-6.2479],[152,-6.2224],[153,-6.2734],[156,-6.7208],[157,-6.7912],[158,-6.8458],[159,-6.8631],[160,-6.8595],[161,-6.8261],[162,-6.7263],[163,-6.5604],[164,-6.4179],[165,-6.4259],[166,-6.847],[167,-8.31],[168,-9.2061],[169,-7.1693],[170,-3.6283],[171,0.7832],[172,3.043],[173,6.5761],[174,9.038],[175,11.1018],[176,12.6715],[177,12.4904],[178,11.1411],[179,9.6439],[180,6.9562],[181,0.6636],[182,-40.8569],[183,-69.7491],[184,-83.0591]]
       }
       arm_l_pv_ik_ctrl.Rotation.Y {
        type=MovieSceneScriptingFloatChannel
        default=82.1786
        thinned_from=185
        keys [[0,66.602],[4,66.6519],[6,66.3637],[8,65.7704],[9,65.6421],[10,65.3211],[14,64.6151],[15,64.6676],[19,65.594],[20,65.7349],[22,66.34],[24,66.7343],[26,66.7645],[27,66.6705],[29,66.0776],[31,65.6712],[32,65.3403],[35,65.1535],[36,64.8937],[38,65.1703],[44,65.5864],[45,65.2628],[46,64.4758],[47,63.3859],[48,62.047],[49,60.1268],[50,57.7005],[51,54.4168],[52,51.9683],[53,51.3631],[54,49.6414],[55,49.8674],[56,46.9669],[57,42.0255],[58,34.9443],[59,27.5962],[60,20.3583],[61,15.0276],[62,11.2607],[63,11.0381],[64,10.9611],[65,10.358],[66,11.0188],[67,11.9464],[68,13.6638],[69,15.5827],[70,17.4378],[71,19.0334],[73,22.7543],[74,24.3512],[75,25.7492],[76,27.0733],[77,27.6217],[78,27.7805],[79,28.5568],[80,29.4972],[81,30.1442],[82,30.504],[83,30.0503],[84,30.0675],[85,30.258],[86,30.6639],[88,31.1708],[91,31.7642],[92,31.9095],[95,32.4356],[97,32.6458],[98,32.5974],[99,32.2848],[100,31.4292],[101,31.2039],[103,31.0394],[104,30.8038],[105,30.667],[106,30.7636],[108,30.6544],[111,30.406],[113,30.0274],[115,29.7352],[119,29.5833],[121,29.5655],[122,29.4754],[126,28.9168],[128,28.7452],[133,28.594],[134,28.4441],[135,28.1724],[136,27.8427],[137,27.678],[140,27.4514],[144,26.8294],[145,26.7168],[150,26.5172],[152,26.2949],[154,25.8608],[155,25.7648],[159,25.735],[162,25.9572],[163,25.9377],[164,25.6458],[165,25.2167],[166,24.2071],[167,21.7678],[168,17.1228],[169,16.7067],[170,19.4801],[171,22.4239],[172,23.7803],[173,24.5315],[174,24.347],[175,23.823],[177,23.1515],[178,23.2306],[179,23.9222],[180,25.6934],[181,31.4806],[182,61.0458],[183,76.7627],[184,82.1786]]
       }
       arm_l_pv_ik_ctrl.Rotation.Z {
        type=MovieSceneScriptingFloatChannel
        default=12.6697
        thinned_from=185
        keys [[0,52.5278],[1,52.4742],[2,52.5395],[4,53.1611],[5,52.868],[6,52.7971],[7,52.029],[8,51.4437],[9,51.3073],[10,50.7914],[11,50.6084],[13,49.8131],[14,49.2965],[15,49.3874],[16,49.9929],[17,50.3752],[18,50.4168],[20,50.86],[21,51.3695],[22,52.0343],[23,52.4281],[24,52.9552],[26,53.049],[27,53.354],[29,52.8517],[31,53.2103],[32,53.0411],[33,53.2475],[34,53.2407],[35,52.7691],[36,52.0247],[38,52.8465],[39,52.5544],[40,52.4085],[41,52.4352],[42,51.8142],[44,51.581],[45,52.0131],[46,52.8375],[47,54.1985],[48,56.5452],[49,60.6263],[50,65.5562],[51,70.9986],[52,75.7049],[53,79.7639],[54,84.3974],[55,88.5137],[56,93.2239],[57,97.4164],[58,99.9473],[59,100.4277],[60,99.4572],[61,97.9171],[62,97.0965],[63,97.4111],[65,99.5522],[66,101.0847],[67,102.3712],[68,103.4445],[71,107.4867],[75,113.8264],[77,116.333],[78,116.89],[82,117.0319],[83,116.2747],[84,115.16],[85,114.6558],[87,115.1704],[96,115.4395],[98,116.9558],[99,117.499],[100,118.295],[101,118.6382],[102,118.6654],[103,118.1957],[104,117.139],[105,116.6042],[106,115.7684],[107,115.4212],[109,114.47],[112,113.652],[115,111.6253],[119,110.5667],[133,108.667],[138,108.724],[143,107.9004],[150,107.6921],[156,107.0851],[163,107.6201],[164,107.5609],[165,107.1145],[166,104.3489],[167,95.4192],[168,72.8286],[169,53.2771],[170,38.4833],[171,26.7418],[172,20.2044],[173,14.0307],[174,11.1653],[175,11.031],[176,12.732],[177,14.5558],[178,16.7122],[179,18.2113],[180,21.6526],[181,26.151],[182,11.6793],[183,4.3117],[184,12.6697]]
       }
       arm_l_pv_ik_ctrl.Scale.X {type=MovieSceneScriptingFloatChannel, default=1, const [1,185]}
       arm_l_pv_ik_ctrl.Scale.Y {type=MovieSceneScriptingFloatChannel, default=1, const [1,185]}
       arm_l_pv_ik_ctrl.Scale.Z {type=MovieSceneScriptingFloatChannel, default=1, const [1,185]}
       upperarm_r_fk_ctrl.Location.X {type=MovieSceneScriptingFloatChannel, default=0.0001, thinned_from=172, keys [[0,0],[165,0],[166,-0.0001],[184,0.0001]]}
       upperarm_r_fk_ctrl.Location.Y {type=MovieSceneScriptingFloatChannel, default=0, thinned_from=172, keys [[0,0],[171,-0.0001],[184,0]]}
       upperarm_r_fk_ctrl.Location.Z {
        type=MovieSceneScriptingFloatChannel
        default=-0.0001
        thinned_from=172
        keys [[0,0],[72,-0.0002],[74,0],[182,0],[183,-0.0003],[184,-0.0001]]
       }
       upperarm_r_fk_ctrl.Rotation.X {
        type=MovieSceneScriptingFloatChannel
        default=74.4624
        thinned_from=171
        keys [[0,55.7091],[1,55.9488],[3,56.1887],[4,56.6859],[5,57.0158],[6,57.0962],[7,57.3712],[8,57.8883],[9,58.2833],[10,58.5199],[12,58.5794],[13,58.8396],[14,58.791],[17,58.0413],[18,57.6343],[23,56.5536],[24,56.4843],[27,56.8608],[29,57.4357],[30,57.8992],[31,58.0385],[33,57.9603],[34,58.0185],[35,58.3298],[38,58.0493],[41,57.9244],[43,58.1491],[44,57.9118],[45,57.4261],[46,56.7022],[51,51.7768],[53,50.0113],[55,48.5234],[56,47.5054],[57,45.9638],[59,42.4752],[60,41.1109],[61,40.3882],[62,40.5987],[63,41.8054],[64,43.6913],[65,45.9776],[66,48.3853],[67,50.6354],[68,52.449],[70,55.2494],[71,56.515],[72,57.6076],[73,58.3894],[75,59.5552],[77,60.235],[78,60.41],[82,60.5475],[83,60.6587],[84,60.5583],[85,60.2771],[86,59.8082],[87,59.6295],[88,59.6088],[90,59.9734],[94,60.0413],[96,60.7034],[107,20.7372],[108,14.8483],[109,10.8178],[110,7.6562],[111,4.5997],[112,1.2524],[113,-1.7195],[114,-3.5957],[115,-5.5705],[116,-6.3862],[117,-6.9777],[118,-7.1606],[119,-7.0354],[120,-6.8349],[121,-6.7292],[122,-6.5779],[123,-6.3748],[124,-6.2568],[125,-6.3122],[126,-6.272],[127,-6.1652],[128,-6.1063],[130,-6.0948],[131,-6.0242],[132,-6.0164],[133,-6.0763],[134,-6.2961],[135,-6.4221],[136,-6.4443],[138,-6.38],[139,-6.3932],[140,-6.4543],[141,-6.574],[142,-6.9498],[143,-7.1305],[144,-7.1401],[145,-7.0954],[147,-7.026],[148,-7.0178],[150,-7.1608],[151,-7.2883],[152,-7.3493],[153,-7.4453],[154,-7.6114],[155,-7.6434],[156,-7.6091],[157,-7.5551],[158,-7.5636],[159,-7.4827],[160,-7.3751],[161,-7.3265],[162,-7.1506],[163,-6.9145],[164,-6.6137],[165,-6.5876],[166,-6.858],[167,-7.0429],[168,-7.5645],[169,-8.0336],[170,-9.6232],[171,-11.2922],[172,-15.7273],[174,-24.8535],[176,-26.0827],[177,-24.3869],[179,-17.8434],[181,-3.0535],[182,-9.2694],[183,46.116],[184,74.4624]]
       }
       upperarm_r_fk_ctrl.Rotation.Y {
        type=MovieSceneScriptingFloatChannel
        default=35.1669
        thinned_from=171
        keys [[0,29.0185],[2,29.0249],[9,29.5445],[12,29.406],[13,29.4315],[15,29.7191],[16,29.7532],[21,29.6338],[23,29.3925],[25,29.2952],[27,29.3581],[28,29.4594],[30,29.8343],[31,29.8627],[33,29.7707],[36,29.8166],[38,29.7851],[41,29.8462],[42,29.912],[43,29.8837],[44,29.7384],[47,29.1394],[48,28.8575],[49,28.4008],[50,27.7824],[51,27.0152],[52,26.1122],[53,25.0863],[54,23.9506],[55,22.7181],[56,20.6478],[57,17.2946],[58,13.1155],[59,8.5672],[60,4.1066],[61,0.1904],[62,-2.7244],[63,-4.7907],[64,-6.5109],[65,-7.9528],[66,-9.184],[67,-10.2721],[68,-11.2848],[69,-12.3532],[71,-14.5717],[72,-15.6069],[73,-16.4509],[74,-17.1965],[75,-17.8161],[76,-18.2448],[77,-18.7457],[78,-18.7628],[79,-18.7379],[82,-18.4726],[83,-18.1568],[84,-18.0983],[85,-18.1684],[86,-18.4254],[88,-18.8081],[90,-18.559],[94,-18.4976],[95,-17.9339],[96,-17.1531],[107,7.5745],[108,9.1601],[109,10.3255],[110,11.4652],[111,12.7699],[112,13.6083],[113,14.5073],[114,15.2554],[115,16.2823],[116,16.7268],[117,17.1017],[119,17.6576],[120,17.826],[122,17.9051],[123,17.8936],[125,17.7606],[126,17.7345],[127,17.7479],[129,17.5903],[130,17.552],[131,17.5964],[133,18.1108],[136,20.3919],[137,20.8034],[138,20.9201],[139,21.214],[140,21.6018],[141,21.7213],[142,21.543],[145,20.6826],[146,20.5584],[148,20.5902],[151,20.8311],[152,20.857],[153,20.795],[155,20.5603],[159,20.2919],[161,20.3742],[162,20.5623],[164,20.8146],[165,20.884],[166,21.0884],[167,22.0432],[168,24.3419],[169,25.3221],[170,24.4624],[171,20.7611],[172,17.8285],[174,11.7122],[176,12.039],[177,11.0513],[179,6.9187],[181,0.3028],[182,26.9042],[183,50.5558],[184,35.1669]]
       }
       upperarm_r_fk_ctrl.Rotation.Z {
        type=MovieSceneScriptingFloatChannel
        default=170.8017
        thinned_from=172
        keys [[0,206.7287],[20,207.5359],[44,207.5887],[45,207.0994],[46,205.7249],[47,203.5692],[48,200.2801],[49,195.5651],[50,189.6181],[51,182.633],[52,174.8036],[53,166.3237],[55,148.1882],[56,137.5053],[57,124.742],[59,97.4957],[60,85.2738],[61,75.4932],[62,69.2846],[63,65.9126],[64,63.7861],[65,62.7522],[66,62.6578],[67,63.35],[68,64.6759],[71,70.0129],[72,71.5991],[73,72.8207],[75,74.8522],[77,76.3462],[78,76.7741],[82,76.6694],[83,76.24],[85,74.9164],[86,74.3822],[88,73.8186],[94,73.952],[96,74.4071],[106,64.6334],[107,64.3689],[108,64.6036],[110,66.1375],[113,67.4397],[114,68.1481],[115,69.7582],[117,71.842],[118,72.6907],[119,73.3252],[120,73.6744],[124,74.0345],[130,73.9076],[133,74.2713],[138,76.1375],[140,76.5372],[141,76.4581],[142,75.9289],[143,75.5641],[145,75.238],[150,75.1847],[153,74.7893],[156,74.2985],[164,74.0768],[165,73.8565],[166,72.7514],[167,69.1246],[168,58.0358],[169,45.0429],[170,31.3112],[171,19.5489],[172,13.9926],[174,10.2017],[176,14.8888],[177,16.7475],[179,21.7675],[181,30.9339],[182,52.1676],[183,140.0843],[184,170.8017]]
       }
       upperarm_r_fk_ctrl.Scale.X {type=MovieSceneScriptingFloatChannel, default=1, const [1,172]}
       upperarm_r_fk_ctrl.Scale.Y {type=MovieSceneScriptingFloatChannel, default=1, const [1,172]}
       upperarm_r_fk_ctrl.Scale.Z {type=MovieSceneScriptingFloatChannel, default=1, const [1,172]}
       lowerarm_r_fk_ctrl.Location.X {
        type=MovieSceneScriptingFloatChannel
        default=-0.0002
        thinned_from=181
        keys [[0,-0.0001],[47,-0.0002],[56,-0.002],[60,-0.0037],[62,-0.0041],[64,-0.0037],[68,-0.0019],[69,-0.0011],[70,-0.0004],[72,-0.001],[74,-0.0001],[97,-0.0002],[99,-0.0005],[100,-0.0014],[101,-0.0011],[102,-0.0019],[103,-0.0008],[104,-0.0007],[105,-0.0002],[106,-0.0005],[107,0],[108,-0.0013],[109,-0.0002],[110,-0.0008],[111,-0.0002],[112,-0.0004],[117,-0.0002],[118,0],[119,-0.0001],[120,0],[121,-0.0002],[122,-0.0001],[125,-0.0002],[130,0],[131,0.0001],[133,-0.0001],[136,0.0001],[138,-0.0002],[139,-0.0002],[140,0.0001],[141,0],[142,0.0001],[143,-0.0003],[146,0],[148,-0.0001],[151,0.0001],[152,-0.0002],[154,0],[155,-0.0002],[156,0.0001],[157,0],[158,0.0002],[159,-0.0002],[160,0.0002],[161,0.0002],[162,-0.0001],[164,-0.0003],[165,0],[166,-0.0036],[167,-0.0001],[168,-0.0044],[169,-0.0057],[170,-0.0092],[171,-0.0079],[172,-0.0012],[174,-0.0002],[182,0],[183,-0.0058],[184,-0.0002]]
       }
       lowerarm_r_fk_ctrl.Location.Y {
        type=MovieSceneScriptingFloatChannel
        default=0.0001
        thinned_from=181
        keys [[0,0],[46,0.0001],[55,0.0019],[60,0.0043],[61,0.0046],[62,0.0048],[64,0.0045],[68,0.0027],[69,0.0014],[70,0.0005],[72,0.0013],[74,0],[75,0.0002],[76,0.0001],[77,0.0002],[99,0.0002],[100,0.0003],[103,0.0001],[104,-0.0003],[105,0],[106,-0.0004],[107,0.0001],[108,-0.0009],[109,-0.0001],[110,-0.0005],[111,0],[112,-0.0003],[113,-0.0001],[114,-0.0001],[115,0.0001],[116,-0.0001],[118,0.0003],[119,0],[120,0.0002],[121,-0.0002],[122,0],[123,-0.0001],[124,0],[125,-0.0002],[130,0.0001],[131,0.0004],[133,-0.0001],[134,0.0002],[135,0.0002],[136,0.0004],[138,-0.0001],[139,-0.0002],[140,0.0003],[141,0.0002],[142,0.0005],[143,-0.0003],[145,-0.0001],[146,0.0003],[148,-0.0002],[149,0.0001],[151,0.0004],[152,-0.0002],[154,0.0002],[155,-0.0001],[156,0.0003],[157,0.0002],[158,0.0006],[159,-0.0002],[160,0.0005],[161,0.0005],[162,-0.0001],[164,-0.0002],[165,0.0002],[166,0.003],[167,0],[168,0.0026],[169,0.0039],[170,0.004],[171,0.002],[172,0.0002],[182,0],[183,0.003],[184,0.0001]]
       }
       lowerarm_r_fk_ctrl.Location.Z {
        type=MovieSceneScriptingFloatChannel
        default=0
        thinned_from=181
        keys [[0,0],[46,-0.0001],[55,-0.0024],[57,-0.0035],[60,-0.0054],[61,-0.0058],[62,-0.006],[63,-0.0058],[64,-0.0053],[68,-0.0023],[69,-0.0012],[70,-0.0004],[72,-0.0011],[74,0],[97,0],[99,-0.0003],[100,-0.0014],[101,-0.0011],[102,-0.0026],[103,-0.0011],[104,-0.0008],[105,-0.0001],[106,-0.0004],[108,0.0004],[109,0.0001],[110,0.0004],[111,-0.0001],[114,0.0003],[115,0.0001],[116,0.0002],[118,-0.0003],[119,0.0001],[120,-0.0002],[121,0.0002],[122,0],[123,0.0002],[124,0],[125,0.0002],[130,0],[131,-0.0003],[133,0.0001],[134,-0.0001],[136,-0.0003],[138,0.0002],[139,0.0002],[140,-0.0003],[141,-0.0001],[142,-0.0004],[143,0.0003],[145,0.0001],[146,-0.0002],[148,0.0002],[149,-0.0001],[151,-0.0003],[152,0.0002],[154,-0.0002],[155,0.0001],[156,-0.0002],[157,-0.0002],[158,-0.0005],[159,0.0002],[160,-0.0004],[161,-0.0004],[162,0.0001],[164,0.0002],[165,-0.0002],[166,0.0015],[167,0.0001],[168,0.0037],[169,0.0056],[170,0.0077],[171,0.0052],[172,0.0004],[174,0],[182,0],[183,0.001],[184,0]]
       }
       lowerarm_r_fk_ctrl.Rotation.X {
        type=MovieSceneScriptingFloatChannel
        default=2.1444
        thinned_from=181
        keys [[0,1.6761],[1,1.6753],[4,1.6486],[5,1.6428],[6,1.6471],[9,1.6431],[10,1.6381],[11,1.6395],[12,1.6471],[13,1.6466],[14,1.6558],[15,1.6614],[17,1.6498],[18,1.6545],[21,1.6796],[22,1.6839],[23,1.6809],[24,1.6875],[25,1.6794],[26,1.6864],[27,1.6841],[28,1.6736],[29,1.672],[30,1.6745],[31,1.67],[33,1.6529],[34,1.6566],[35,1.6521],[38,1.6697],[41,1.6626],[43,1.6714],[44,1.6729],[45,1.6792],[46,1.693],[47,1.7136],[48,1.7404],[49,1.7727],[50,1.8096],[52,1.8946],[55,2.0395],[56,2.106],[57,2.1968],[59,2.4013],[60,2.4896],[61,2.5519],[62,2.5755],[63,2.5251],[64,2.3927],[65,2.2065],[67,1.786],[68,1.6081],[69,1.467],[70,1.3517],[71,1.2673],[72,1.2027],[73,1.1499],[74,1.1046],[75,1.0624],[76,1.0393],[77,1.0306],[78,1.0316],[79,1.0271],[80,1.015],[81,0.998],[82,0.9784],[83,0.9141],[84,0.8888],[85,0.8905],[86,0.9233],[87,0.9308],[88,0.9246],[90,0.9361],[95,0.9372],[96,0.9329],[97,0.9033],[98,0.8385],[99,0.7575],[100,0.6245],[101,0.4861],[102,0.3941],[103,0.3951],[104,0.7166],[105,0.9034],[106,1.3037],[107,1.5482],[108,2.0786],[109,2.4647],[110,2.6125],[111,2.5498],[112,2.3144],[113,1.8546],[114,1.2924],[115,0.2387],[116,-0.3859],[117,-1.0341],[118,-1.5297],[119,-1.9712],[120,-2.2743],[121,-2.4215],[122,-2.6161],[123,-2.7854],[124,-2.9273],[125,-2.9984],[127,-2.9803],[128,-3.0015],[130,-3.0149],[131,-3.0116],[133,-2.9772],[134,-2.8466],[135,-2.4956],[136,-1.644],[137,-0.597],[138,0.5722],[139,1.2897],[140,1.6654],[141,1.9007],[142,1.9917],[145,2.0092],[146,2.007],[148,2.0122],[150,1.9986],[152,1.9724],[153,1.9498],[154,1.9207],[157,1.8907],[158,1.8768],[159,1.871],[160,1.8571],[161,1.8473],[162,1.8238],[164,1.7635],[165,1.7163],[166,1.577],[167,1.197],[168,0.0643],[169,-0.8148],[170,-0.8513],[171,-0.1815],[172,0.3634],[174,1.0956],[176,1.0982],[177,1.039],[179,0.9656],[181,1.0368],[182,1.1162],[183,3.3556],[184,2.1444]]
       }
       lowerarm_r_fk_ctrl.Rotation.Y {
        type=MovieSceneScriptingFloatChannel
        default=-0.3656
        thinned_from=181
        keys [[0,0.8621],[1,0.8641],[2,0.8794],[3,0.8975],[4,0.9185],[5,0.9303],[6,0.9215],[7,0.9265],[9,0.9298],[10,0.9394],[11,0.9359],[12,0.9205],[13,0.9213],[14,0.9023],[15,0.8904],[17,0.913],[18,0.9039],[19,0.8848],[21,0.8531],[22,0.8444],[23,0.8509],[24,0.837],[25,0.8539],[26,0.8394],[27,0.8445],[28,0.8669],[29,0.8704],[30,0.8654],[31,0.8743],[33,0.9089],[34,0.9016],[35,0.9106],[36,0.8989],[37,0.8842],[38,0.8742],[39,0.8765],[41,0.8879],[42,0.8798],[43,0.8689],[44,0.8654],[45,0.8041],[46,0.6378],[47,0.3767],[48,0.0315],[49,-0.3877],[50,-0.8702],[51,-1.4058],[52,-1.984],[53,-2.5943],[54,-3.2265],[55,-3.87],[56,-4.7727],[57,-6.0746],[58,-7.6039],[59,-9.1889],[60,-10.6578],[61,-11.8386],[62,-12.5596],[63,-12.953],[64,-13.2784],[65,-13.5556],[68,-14.295],[69,-14.5087],[70,-14.6618],[71,-14.7762],[75,-15.0419],[82,-15.1461],[83,-15.2255],[84,-15.2562],[85,-15.2542],[86,-15.2142],[96,-15.2026],[97,-15.2386],[99,-15.4112],[101,-15.7124],[102,-15.8084],[103,-15.8055],[104,-15.4529],[105,-15.2406],[106,-14.7264],[107,-14.3897],[108,-13.5146],[109,-12.7829],[110,-12.4536],[111,-12.6211],[112,-13.1336],[114,-14.7697],[115,-15.9411],[116,-16.4864],[117,-16.9628],[118,-17.2734],[119,-17.5179],[120,-17.6686],[124,-17.9458],[133,-17.968],[134,-17.9157],[135,-17.7693],[136,-17.3384],[137,-16.6512],[138,-15.606],[139,-14.767],[140,-14.2418],[141,-13.8785],[142,-13.7296],[148,-13.6947],[152,-13.766],[154,-13.8519],[161,-13.9699],[164,-14.1013],[165,-14.1717],[166,-14.3704],[167,-14.8916],[168,-16.0585],[169,-16.6802],[170,-16.635],[171,-16.1263],[172,-15.6794],[174,-15.0106],[176,-15.0094],[179,-15.1454],[181,0.4283],[182,0.1121],[183,-7.2231],[184,-0.3656]]
       }
       lowerarm_r_fk_ctrl.Rotation.Z {
        type=MovieSceneScriptingFloatChannel
        default=-14.8692
        thinned_from=181
        keys [[0,-19.7207],[1,-19.7281],[4,-19.959],[5,-20.0089],[6,-19.9721],[9,-20.0066],[10,-20.049],[11,-20.0364],[12,-19.9709],[13,-19.975],[15,-19.8468],[17,-19.9458],[18,-19.9061],[21,-19.6882],[22,-19.6507],[23,-19.6778],[24,-19.6195],[25,-19.6904],[26,-19.6291],[27,-19.6496],[28,-19.7424],[30,-19.7347],[33,-19.9214],[34,-19.8898],[35,-19.928],[38,-19.7754],[41,-19.8366],[44,-19.7454],[45,-19.4166],[46,-18.5093],[47,-17.081],[48,-15.1892],[49,-12.8913],[50,-10.2449],[51,-7.3074],[52,-4.1362],[53,-0.7889],[54,2.677],[55,6.2041],[56,11.1768],[57,18.3867],[58,26.8876],[59,35.7333],[60,43.9774],[61,50.6736],[62,54.8758],[63,57.2662],[64,59.24],[65,60.9204],[68,65.4346],[69,66.8061],[70,67.8312],[71,68.5826],[73,69.6054],[75,70.3801],[76,70.5794],[79,70.6145],[82,71.0619],[83,71.6773],[84,71.9276],[85,71.9064],[86,71.5921],[96,71.5043],[97,71.7782],[99,73.0832],[101,75.3877],[102,76.1283],[103,76.1769],[104,73.6919],[105,72.241],[106,68.3758],[107,66.1245],[108,60.5558],[109,56.1087],[110,54.202],[111,55.2196],[112,58.3443],[113,63.5558],[114,69.0955],[115,78.008],[117,87.3807],[118,90.7711],[119,93.7034],[120,95.669],[121,96.6237],[123,98.9384],[124,99.826],[125,100.2739],[127,100.1553],[131,100.3407],[133,100.1314],[134,99.3139],[135,97.0917],[136,91.5342],[137,84.2665],[138,75.2857],[139,69.0694],[140,65.4929],[141,63.1174],[142,62.1815],[148,61.9387],[152,62.4253],[154,62.9961],[161,63.7725],[164,64.6205],[165,65.0855],[166,66.4422],[167,70.19],[168,80.0734],[169,84.286],[170,79.2568],[171,66.5244],[172,53.104],[174,31.2351],[176,27.6308],[177,30.1813],[179,38.5416],[181,53.9517],[182,70.6592],[183,21.2053],[184,-14.8692]]
       }
       lowerarm_r_fk_ctrl.Scale.X {type=MovieSceneScriptingFloatChannel, default=1, const [1,181]}
       lowerarm_r_fk_ctrl.Scale.Y {type=MovieSceneScriptingFloatChannel, default=1, const [1,181]}
       lowerarm_r_fk_ctrl.Scale.Z {type=MovieSceneScriptingFloatChannel, default=1, const [1,181]}
       hand_r_fk_ctrl.Location.X {
        type=MovieSceneScriptingFloatChannel
        default=-0.0004
        thinned_from=185
        keys [[0,-0.0001],[44,-0.0001],[46,-0.0005],[49,-0.0019],[55,-0.0067],[56,-0.0083],[57,-0.0112],[59,-0.0184],[60,-0.0217],[61,-0.0241],[62,-0.025],[63,-0.0241],[64,-0.0216],[65,-0.0181],[67,-0.0102],[69,-0.0033],[70,-0.0013],[72,-0.0025],[74,-0.0001],[75,-0.0003],[76,-0.0001],[77,-0.0003],[97,-0.0002],[99,-0.0004],[100,-0.0017],[101,-0.0013],[102,-0.0047],[103,-0.002],[104,-0.0036],[105,-0.0007],[106,-0.0039],[107,-0.0001],[108,-0.0087],[109,-0.0005],[110,-0.0026],[111,-0.0012],[112,-0.0028],[113,-0.003],[114,-0.0039],[116,-0.0024],[117,-0.0022],[118,-0.001],[119,-0.0009],[120,-0.0003],[121,-0.0004],[122,-0.0001],[123,-0.0003],[133,-0.0001],[134,-0.0007],[135,-0.0002],[136,-0.0084],[137,-0.0001],[138,-0.0073],[139,-0.0004],[140,-0.002],[141,-0.0002],[142,-0.0004],[143,-0.0001],[165,-0.0001],[166,-0.0014],[167,-0.0001],[168,-0.0024],[169,-0.005],[170,-0.0095],[171,-0.0085],[172,-0.001],[173,0],[174,-0.0002],[175,0],[177,-0.0002],[178,0],[179,-0.0001],[182,0],[183,-0.0236],[184,-0.0004]]
       }
       hand_r_fk_ctrl.Location.Y {
        type=MovieSceneScriptingFloatChannel
        default=0
        thinned_from=185
        keys [[0,0],[47,0.0002],[55,0.0017],[56,0.0022],[60,0.0057],[61,0.0064],[62,0.0066],[63,0.0064],[64,0.0058],[68,0.0022],[69,0.0011],[70,0.0004],[72,0.0007],[74,0],[99,-0.0001],[100,-0.0003],[101,-0.0002],[102,-0.0006],[103,-0.0003],[104,-0.0005],[105,-0.0001],[106,-0.0004],[107,0],[108,-0.0006],[109,-0.0001],[110,-0.0003],[111,0.0002],[112,0.0004],[113,0.0009],[114,0.0011],[117,0.0006],[118,0.0003],[134,-0.0001],[135,-0.0001],[136,0.0007],[137,0],[138,0.0007],[139,0],[167,0],[168,0.0002],[169,0.0018],[170,0.0053],[171,0.0059],[172,0.0008],[173,-0.431],[174,0.0001],[175,-0.431],[176,0],[177,0],[178,-0.431],[179,0],[180,-0.431],[182,-0.431],[183,0.0008],[184,0]]
       }
       hand_r_fk_ctrl.Location.Z {
        type=MovieSceneScriptingFloatChannel
        default=0
        thinned_from=185
        keys [[0,0],[47,-0.0002],[55,-0.0016],[56,-0.002],[60,-0.0051],[61,-0.0057],[62,-0.0059],[63,-0.0057],[64,-0.0052],[68,-0.002],[69,-0.001],[70,-0.0003],[72,-0.0008],[74,0],[99,0],[100,0.0004],[101,0.0003],[102,0.0009],[103,0.0003],[104,0.0006],[105,0.0001],[106,0.0004],[107,0],[108,0.0008],[109,0.0001],[110,0.0005],[111,0.0002],[112,0.0006],[113,0.0002],[114,0.0004],[115,0],[116,0.0001],[117,0],[135,0],[136,0.0002],[137,0],[167,0],[168,0.0007],[170,0.0037],[171,0.0023],[172,0.0002],[173,0.0668],[174,0.0001],[175,0.0668],[176,0],[177,0],[178,0.0668],[179,0],[180,0.0668],[182,0.0668],[183,-0.0017],[184,0]]
       }
       hand_r_fk_ctrl.Rotation.X {
        type=MovieSceneScriptingFloatChannel
        default=12.2504
        thinned_from=185
        keys [[0,3.9183],[1,3.9717],[2,4.0156],[3,4.0061],[4,4.334],[5,4.6736],[6,4.8165],[7,4.915],[8,5.1282],[9,5.27],[10,5.2422],[11,5.2789],[13,5.217],[14,5.051],[15,4.8266],[16,4.6764],[17,4.4573],[18,4.152],[19,4.061],[20,4.0972],[21,4.0053],[22,3.8477],[23,3.8734],[24,3.8307],[25,3.8197],[26,3.9266],[27,4.1697],[28,4.1875],[29,4.3538],[30,4.5037],[31,4.5746],[32,4.5123],[33,4.418],[34,4.4224],[35,4.3819],[36,4.4141],[37,4.468],[38,4.4412],[39,4.4549],[40,4.4846],[42,4.3154],[43,4.3367],[44,4.3069],[46,4.3196],[48,4.3544],[52,4.4707],[55,4.5795],[56,4.6494],[57,4.7698],[59,5.073],[60,5.211],[61,5.3102],[62,5.3482],[63,5.1095],[64,4.4773],[65,3.5781],[66,2.5379],[67,1.4832],[68,0.5401],[69,-0.6297],[70,-1.4136],[72,-1.5119],[73,-2.3025],[74,-3.4228],[75,-3.9116],[76,-4.0164],[77,-3.5103],[78,-4.6074],[79,-5.2224],[81,-6.2945],[82,-6.8909],[83,-7.864],[84,-7.8447],[85,-7.9389],[86,-7.6857],[87,-7.6001],[88,-7.4926],[89,-7.6466],[90,-7.7],[94,-7.6956],[95,-8.257],[96,-8.3695],[97,-8.9868],[98,-10.2839],[99,-12.9702],[100,-21.4618],[101,-23.2683],[102,-21.4885],[103,-27.3237],[104,-21.6708],[105,-19.1842],[106,-13.1184],[107,-8.6483],[108,-1.4244],[109,7.7599],[110,18.964],[111,26.7935],[112,17.3527],[113,14.0086],[114,14.3036],[115,14.9145],[116,13.8794],[117,13.0425],[118,13.2041],[120,13.314],[121,13.4737],[122,14.0273],[123,14.1512],[124,14.3329],[125,14.0033],[126,13.8444],[127,14.0004],[128,13.5638],[129,13.5956],[130,13.7832],[131,13.7303],[132,13.9782],[133,14.2833],[134,16.166],[135,17.2737],[136,17.3913],[137,17.3008],[138,17.6342],[139,18.5358],[140,19.3677],[141,19.1763],[142,18.9049],[143,18.7055],[144,18.7712],[145,18.8771],[146,18.8691],[147,18.781],[148,18.8176],[149,18.9014],[151,18.7398],[152,18.7604],[154,18.5653],[155,18.7123],[156,18.8069],[157,18.7877],[158,18.7101],[159,18.8159],[161,18.6879],[162,18.7053],[163,18.8172],[165,18.7878],[166,18.9836],[167,18.3646],[168,16.2733],[169,17.0174],[170,19.3169],[171,20.9711],[172,23.4292],[173,29.6691],[174,28.1833],[175,28.1688],[176,29.2681],[177,29.1168],[178,28.5681],[179,29.3642],[180,30.0636],[181,0],[182,0],[183,17.5848],[184,12.2504]]
       }
       hand_r_fk_ctrl.Rotation.Y {
        type=MovieSceneScriptingFloatChannel
        default=6.6226
        thinned_from=185
        keys [[0,7.5189],[1,7.4822],[2,7.5697],[4,7.4892],[5,7.6521],[6,7.4786],[7,7.6246],[8,7.7214],[10,7.7798],[11,7.7477],[12,7.6171],[13,7.5715],[14,7.6614],[15,7.6415],[16,7.6795],[17,7.5209],[18,7.5996],[19,7.5478],[20,7.5327],[21,7.6704],[22,7.5396],[23,7.6821],[24,7.7618],[25,7.7644],[26,7.628],[27,7.7848],[28,7.8071],[29,7.7464],[30,7.8646],[31,7.9159],[32,7.8882],[33,7.8212],[34,7.6633],[36,7.9544],[37,7.9381],[38,7.9756],[39,8.0577],[40,7.9634],[41,7.9055],[42,7.8715],[43,8.0105],[44,8.0036],[45,7.9356],[46,7.7591],[47,7.4848],[48,7.1234],[49,6.6855],[50,6.1818],[51,5.623],[52,5.0196],[53,4.3824],[54,3.722],[55,3.0491],[56,1.7802],[57,-0.4053],[58,-3.1022],[59,-5.9051],[60,-8.4089],[61,-10.2083],[62,-10.8982],[63,-10.8542],[64,-10.7328],[65,-10.5494],[66,-10.3194],[68,-9.7821],[69,-9.3211],[70,-8.5303],[71,-6.9177],[72,-5.5754],[73,-5.274],[75,-4.8107],[76,-4.8322],[77,-5.1525],[78,-5.657],[79,-5.627],[81,-5.4947],[82,-5.4646],[83,-5.6374],[84,-5.6771],[85,-5.5228],[86,-5.269],[87,-5.3382],[88,-5.4637],[89,-5.2175],[90,-5.159],[91,-5.131],[94,-5.1219],[95,-4.8784],[96,-4.1947],[97,-4.0234],[98,-3.5829],[99,-1.736],[100,-0.1267],[101,1.5335],[102,3.5499],[103,2.6308],[104,12.587],[105,14.3983],[106,4.2484],[107,2.2789],[108,-3.9901],[109,-8.9001],[110,-12.7212],[111,-11.7889],[112,-11.4346],[113,-5.7486],[114,-1.4401],[115,3.1825],[116,4.2824],[117,4.2518],[118,3.9427],[119,3.7273],[120,3.49],[121,3.1768],[122,2.9975],[123,2.8796],[124,2.176],[125,1.2873],[126,1.5185],[127,1.6883],[128,1.422],[129,1.4482],[130,1.4352],[131,1.4556],[133,1.7044],[134,2.5888],[135,2.9519],[136,2.9458],[137,2.3441],[138,3.4724],[139,4.4272],[140,4.675],[141,5.0522],[142,5.3537],[143,5.3726],[144,5.4473],[146,5.363],[147,5.3451],[148,5.4364],[149,5.6792],[150,5.6559],[151,5.6555],[152,5.7381],[153,5.6999],[154,5.675],[155,5.6699],[156,5.6188],[157,5.5966],[158,5.4224],[159,5.3975],[160,5.2306],[161,5.086],[162,4.9226],[163,4.7909],[164,4.905],[165,4.87],[166,4.6453],[167,3.6737],[168,1.7084],[169,4.2734],[170,8.2247],[171,10.6209],[172,12.3088],[173,-10.7647],[174,16.6015],[175,16.8861],[176,16.6183],[177,16.7978],[178,15.4637],[179,19.34],[180,18.3882],[181,0],[182,0],[183,-1.5933],[184,6.6226]]
       }
       hand_r_fk_ctrl.Rotation.Z {
        type=MovieSceneScriptingFloatChannel
        default=-34.4326
        thinned_from=185
        keys [[0,-32.9971],[1,-32.3815],[2,-32.0932],[4,-31.3483],[5,-31.1215],[7,-31.0077],[9,-30.3235],[10,-30.2036],[11,-30.1688],[14,-30.6532],[15,-31.0504],[17,-31.5286],[18,-32.0552],[19,-32.4458],[20,-32.4329],[22,-33.102],[23,-33.1317],[24,-33.0753],[25,-33.1839],[29,-32.2938],[30,-32.0112],[32,-32.2501],[33,-32.1817],[34,-32.0206],[35,-32.0764],[36,-32.2003],[37,-32.1336],[40,-32.4227],[42,-32.437],[43,-32.5523],[44,-32.7429],[45,-32.5851],[46,-32.1316],[47,-31.4118],[48,-30.4554],[49,-29.292],[50,-27.9511],[51,-26.4622],[52,-24.8551],[53,-23.1592],[55,-19.6196],[56,-16.248],[57,-10.4302],[58,-3.243],[59,4.2369],[60,10.9327],[61,15.7675],[62,17.6646],[66,17.9008],[68,17.9419],[70,17.7044],[71,17.8937],[72,18.295],[73,18.8177],[74,19.4009],[76,20.4453],[77,20.725],[78,21.473],[80,22.6759],[81,23.1579],[82,23.3532],[83,23.2737],[84,22.5736],[85,22.6924],[86,23.7997],[87,23.6411],[89,23.2028],[90,23.0676],[91,23],[94,22.9815],[95,21.9832],[97,19.2794],[98,17.216],[99,15.9192],[100,16.2322],[101,14.3132],[102,18.254],[103,15.0774],[104,18.8128],[106,25.2317],[107,26.5854],[108,24.7695],[109,18.9113],[110,9.8912],[111,-2.1205],[112,-7.1874],[113,-14.0694],[114,-18.6672],[115,-24.2917],[116,-26.3628],[117,-27.0301],[118,-26.7033],[119,-26.8792],[120,-27.3897],[121,-27.616],[122,-27.7336],[123,-28.039],[124,-28.4992],[125,-28.6328],[127,-28.6923],[128,-29.1345],[129,-29.4864],[130,-29.4237],[131,-29.4895],[133,-29.0157],[134,-28.417],[135,-28.976],[136,-30.7767],[137,-32.5157],[138,-34.6171],[139,-35.5013],[140,-35.0819],[141,-35.2565],[142,-35.2687],[143,-34.9622],[144,-34.7691],[145,-34.905],[148,-34.8482],[150,-35.0522],[151,-34.8671],[153,-35.0389],[155,-34.8682],[156,-35.0003],[157,-35.0171],[158,-34.9184],[162,-34.8586],[163,-34.7828],[165,-34.7622],[166,-35.0284],[167,-36.8257],[168,-42.4235],[169,-51.7108],[170,-53.6383],[171,-48.652],[172,-40.4453],[173,-27.1908],[174,-26.3698],[175,-26.6331],[176,-23.7505],[177,-25.6652],[178,-18.6621],[179,-31.3638],[180,-22.6546],[181,0],[182,0],[183,-32.4598],[184,-34.4326]]
       }
       hand_r_fk_ctrl.Scale.X {type=MovieSceneScriptingFloatChannel, default=1, const [1,185]}
       hand_r_fk_ctrl.Scale.Y {type=MovieSceneScriptingFloatChannel, default=1, const [1,185]}
       hand_r_fk_ctrl.Scale.Z {type=MovieSceneScriptingFloatChannel, default=1, const [1,185]}
       thumb_01_r_ctrl.Location.X {
        type=MovieSceneScriptingFloatChannel
        default=-0.0002
        thinned_from=185
        keys [[0,-0.0001],[55,-0.0006],[61,-0.002],[63,-0.002],[70,-0.0002],[72,-0.0005],[74,-0.0001],[101,-0.0003],[102,-0.0015],[103,-0.0007],[104,-0.0002],[105,-0.0002],[106,-0.0009],[107,-0.0001],[108,-0.0051],[109,-0.0008],[110,-0.0036],[111,-0.0011],[112,-0.0027],[113,-0.0003],[133,-0.0001],[135,-0.0002],[136,-0.0011],[137,-0.0001],[138,-0.0009],[139,-0.0002],[140,-0.0003],[141,-0.0001],[168,-0.0001],[170,-0.0006],[171,-0.0006],[172,-0.0002],[180,-0.0002],[182,-0.0005],[183,-0.0009],[184,-0.0002]]
       }
       thumb_01_r_ctrl.Location.Y {
        type=MovieSceneScriptingFloatChannel
        default=0.0001
        thinned_from=185
        keys [[0,0.0001],[55,-0.0005],[61,-0.002],[63,-0.002],[69,-0.0002],[99,0],[101,0.0001],[102,-0.0001],[105,0.0001],[106,-0.0003],[107,0.0001],[108,-0.0012],[109,-0.0001],[110,-0.0013],[111,-0.0004],[112,-0.0011],[113,-0.0002],[135,0.0001],[136,-0.0006],[137,0.0001],[138,-0.0004],[139,0.0001],[140,-0.0001],[169,0],[171,-0.0002],[172,0.0001],[181,0.0001],[183,-0.0003],[184,0.0001]]
       }
       thumb_01_r_ctrl.Location.Z {
        type=MovieSceneScriptingFloatChannel
        default=0.0001
        thinned_from=185
        keys [[0,0.0001],[56,-0.0002],[62,-0.0009],[70,0],[101,0],[102,-0.0008],[103,-0.0002],[104,0.0001],[107,0.0001],[108,0.0008],[109,0.0003],[110,0.0015],[111,-0.0005],[112,-0.0012],[113,-0.0008],[114,-0.0011],[115,-0.0004],[135,0.0001],[136,-0.0004],[137,0.0001],[138,-0.0001],[139,0.0001],[168,0],[169,-0.0001],[170,-0.0005],[171,-0.0006],[172,0],[180,0.0001],[181,0],[183,0.0003],[184,0.0001]]
       }
       thumb_01_r_ctrl.Rotation.X {
        type=MovieSceneScriptingFloatChannel
        default=5.5578
        thinned_from=185
        keys [[0,4.0168],[1,4.0264],[2,3.9751],[3,4.1362],[4,4.038],[5,4.2271],[6,4.1897],[7,4.3776],[8,4.392],[9,4.4797],[10,4.4692],[11,4.4934],[13,4.4222],[14,4.3504],[15,4.3811],[16,4.4887],[17,4.5341],[18,4.6054],[19,4.5644],[20,4.5733],[21,4.6327],[22,4.6415],[23,4.5883],[24,4.7066],[25,4.763],[26,4.6594],[27,4.6792],[29,4.6636],[30,4.7294],[31,4.8572],[32,4.8341],[33,4.9051],[34,4.718],[35,4.9441],[36,4.8794],[37,4.7751],[38,4.684],[39,4.738],[40,4.8044],[41,4.7273],[42,4.7825],[43,4.9349],[44,4.8598],[55,4.755],[56,4.7303],[57,4.6888],[60,4.523],[61,4.4739],[62,4.4378],[67,4.3511],[68,4.3276],[69,4.265],[70,4.1494],[71,3.8679],[72,3.6299],[73,3.5871],[74,3.5676],[75,3.6152],[76,3.6023],[77,3.3791],[78,3.6335],[82,3.6397],[83,3.272],[84,3.038],[85,2.9339],[86,3.0667],[87,3.1096],[88,3.2386],[90,3.2303],[91,3.124],[92,2.6919],[93,2.974],[94,3.107],[95,2.7878],[96,2.6417],[97,2.3251],[98,2.1416],[99,2.1116],[100,3.3748],[101,1.1586],[102,-1.7725],[103,2.2662],[104,4.6708],[105,4.4473],[106,2.5477],[107,3.1477],[108,2.3218],[109,2.3003],[110,2.3552],[111,3.122],[112,5.8696],[113,4.8386],[114,4.5289],[115,4.5781],[116,4.6922],[117,4.5719],[118,4.4778],[119,4.4293],[120,4.4252],[121,4.2152],[122,4.0622],[123,4.0388],[124,3.7616],[125,3.63],[126,3.6398],[127,3.5609],[128,3.6247],[129,3.5996],[130,3.452],[131,3.4631],[132,3.5123],[133,3.3244],[134,3.4721],[135,3.4212],[136,2.5382],[137,2.8353],[138,2.5717],[139,2.5554],[140,2.7094],[141,2.8918],[142,2.661],[143,2.5934],[144,2.6574],[145,2.4968],[146,2.4808],[147,2.5973],[148,2.6329],[149,2.675],[150,2.6425],[151,2.7633],[152,2.7442],[153,2.8319],[154,2.8753],[155,2.8052],[156,2.8326],[157,2.8489],[158,2.7747],[159,2.8581],[160,2.9016],[161,2.9252],[162,2.8997],[163,2.7743],[164,2.8395],[165,2.8431],[166,2.7917],[167,3.1004],[168,4.4237],[169,6.2588],[170,7.6237],[171,8.3718],[172,8.358],[173,8.7431],[174,8.332],[175,8.3773],[176,7.8777],[177,7.9622],[178,7.8998],[180,7.6648],[181,7.8316],[182,6.476],[183,4.948],[184,5.5578]]
       }
       thumb_01_r_ctrl.Rotation.Y {
        type=MovieSceneScriptingFloatChannel
        default=9.0465
        thinned_from=185
        keys [[0,8.6857],[1,8.588],[2,8.6457],[3,8.5691],[4,8.565],[5,8.6839],[6,8.677],[7,8.6044],[8,8.6194],[10,8.5964],[11,8.618],[12,8.5908],[13,8.6119],[14,8.5797],[15,8.6928],[17,8.5635],[18,8.5616],[19,8.5961],[20,8.5494],[21,8.6407],[22,8.5621],[23,8.6399],[24,8.5583],[25,8.6127],[26,8.5521],[28,8.6025],[29,8.5405],[31,8.6142],[32,8.6283],[34,8.5715],[35,8.5825],[36,8.5196],[37,8.503],[38,8.5414],[39,8.54],[40,8.6283],[41,8.6436],[42,8.5304],[44,8.6427],[55,8.5348],[56,8.5054],[58,8.3834],[61,8.1692],[65,8.0296],[67,7.9924],[69,7.9962],[70,8.0161],[72,8.8745],[73,8.826],[74,8.7502],[76,8.6794],[77,8.5182],[78,8.3201],[80,8.243],[82,8.208],[83,8.321],[84,8.3568],[85,8.3125],[86,8.4474],[87,8.4514],[88,8.5468],[90,8.3908],[91,8.3309],[92,8.0136],[93,8.1061],[94,8.1554],[95,8.0741],[96,8.1252],[97,8.0807],[99,8.1942],[100,6.9753],[101,6.7454],[102,6.2597],[103,4.8179],[104,4.8598],[105,5.396],[106,6.2504],[107,7.0472],[108,8.0836],[109,8.3258],[110,8.1804],[111,7.2495],[112,6.7456],[113,6.9278],[114,6.7669],[115,5.7775],[116,5.5362],[117,5.464],[118,5.4331],[119,5.3812],[120,5.5195],[121,5.5597],[122,5.6567],[123,5.7236],[124,5.69],[125,5.8449],[126,5.8928],[127,5.8794],[128,5.9625],[129,5.968],[131,5.9079],[132,5.9637],[133,5.9742],[134,6.1019],[135,6.2453],[136,6.4777],[137,6.3635],[138,6.3519],[139,6.2183],[140,6.2037],[141,6.1643],[142,6.3018],[143,6.3425],[144,6.4002],[145,6.4278],[146,6.3853],[147,6.4143],[148,6.4243],[149,6.3647],[151,6.1921],[152,6.178],[153,6.1894],[154,6.2359],[155,6.2463],[156,6.2846],[157,6.2779],[158,6.2973],[159,6.2751],[160,6.2881],[161,6.2651],[162,6.3325],[163,6.3095],[164,6.355],[166,6.3852],[167,6.4818],[168,6.4983],[169,7.4239],[170,8.121],[171,9.3232],[172,9.894],[173,10.1469],[174,10.3765],[175,10.1758],[177,10.0769],[179,9.7818],[180,9.615],[181,9.1509],[182,8.3966],[183,8.6106],[184,9.0465]]
       }
       thumb_01_r_ctrl.Rotation.Z {
        type=MovieSceneScriptingFloatChannel
        default=-9.1275
        thinned_from=185
        keys [[0,-9.2654],[1,-9.3132],[3,-9.354],[4,-9.296],[6,-9.2341],[7,-9.2216],[8,-9.2553],[10,-9.1899],[11,-9.1897],[12,-9.1344],[13,-9.1583],[14,-9.1047],[16,-9.1464],[17,-9.086],[18,-9.2179],[19,-9.2336],[20,-9.2728],[21,-9.2303],[23,-9.2059],[24,-9.2286],[25,-9.2048],[26,-9.2393],[27,-9.195],[28,-9.2776],[29,-9.1687],[30,-9.1412],[31,-9.1865],[32,-9.1557],[33,-9.1707],[35,-9.115],[36,-9.1987],[37,-9.1638],[38,-9.2515],[39,-9.1629],[40,-9.1897],[41,-9.1403],[42,-9.1185],[43,-9.149],[44,-9.1367],[45,-9.1471],[47,-9.2249],[49,-9.3654],[52,-9.6594],[55,-10.0063],[56,-10.2405],[57,-10.6575],[58,-11.184],[59,-11.747],[60,-12.2733],[61,-12.69],[62,-12.9239],[64,-13.1179],[69,-13.4811],[70,-13.5278],[72,-12.7005],[73,-12.7638],[74,-12.8653],[75,-12.93],[76,-13.037],[77,-13.1823],[78,-13.1652],[79,-13.2064],[81,-13.3972],[82,-13.4537],[83,-13.4674],[84,-13.5182],[85,-13.4899],[86,-13.5213],[88,-13.3883],[90,-13.4818],[91,-13.4379],[92,-13.666],[93,-13.5872],[94,-13.7082],[95,-13.537],[96,-13.5584],[97,-13.5083],[98,-13.6716],[99,-13.6273],[100,-16.7617],[101,-14.4978],[102,-14.6683],[103,-18.389],[104,-16.9744],[105,-16.0676],[106,-16.1155],[107,-14.7804],[108,-13.2035],[109,-12.8232],[110,-12.6162],[111,-12.688],[112,-15.8373],[113,-16.2475],[114,-16.575],[115,-17.2395],[116,-17.5295],[117,-17.8979],[118,-18.1955],[119,-18.6549],[120,-18.8646],[122,-18.8958],[123,-18.8289],[124,-18.9024],[125,-18.6994],[127,-18.6357],[128,-18.4557],[129,-18.4566],[130,-18.3685],[131,-18.4937],[132,-18.4877],[133,-18.3809],[134,-18.1556],[135,-17.7258],[136,-16.8609],[137,-16.2945],[138,-16.1442],[139,-16.2159],[140,-16.4116],[141,-16.3406],[142,-16.383],[143,-16.3357],[144,-16.2292],[145,-16.0294],[148,-15.9962],[149,-16.0624],[152,-16.4449],[153,-16.5309],[154,-16.5194],[157,-16.3929],[159,-16.2453],[160,-16.2297],[161,-16.3267],[163,-16.2612],[165,-16.121],[166,-15.8341],[167,-14.1556],[168,-11.2995],[169,-10.4575],[170,-9.106],[171,-8.6023],[172,-8.0703],[173,-7.8553],[174,-7.7561],[175,-8.0389],[176,-8.3609],[177,-8.3761],[178,-8.1869],[179,-8.3688],[180,-8.6464],[181,-9.1328],[182,-9.8954],[183,-9.5695],[184,-9.1275]]
       }
       thumb_01_r_ctrl.Scale.X {type=MovieSceneScriptingFloatChannel, default=1, const [1,185]}
       thumb_01_r_ctrl.Scale.Y {type=MovieSceneScriptingFloatChannel, default=1, const [1,185]}
       thumb_01_r_ctrl.Scale.Z {type=MovieSceneScriptingFloatChannel, default=1, const [1,185]}
       thumb_02_r_ctrl.Location.X {
        type=MovieSceneScriptingFloatChannel
        default=0.0001
        thinned_from=185
        keys [[0,0.0001],[70,0.0001],[72,-0.0003],[74,0.0001],[82,-0.0001],[97,0.0001],[99,-0.0001],[100,-0.0005],[101,-0.0003],[102,-0.0022],[103,-0.0008],[104,0],[105,0.0001],[106,-0.0012],[107,0.0001],[108,-0.011],[109,-0.0016],[110,-0.009],[111,-0.001],[112,-0.0026],[113,-0.0022],[114,-0.003],[115,-0.001],[116,-0.0007],[117,-0.0001],[133,0.0001],[134,0],[135,0.0001],[136,-0.0017],[137,0.0001],[138,-0.0015],[139,0],[140,-0.0004],[141,0.0001],[165,0.0001],[166,0],[167,0.0001],[169,-0.0002],[170,-0.0012],[171,-0.0016],[172,0],[180,0.0001],[181,-0.0005],[182,-0.0007],[183,-0.0017],[184,0.0001]]
       }
       thumb_02_r_ctrl.Location.Y {
        type=MovieSceneScriptingFloatChannel
        default=0
        thinned_from=185
        keys [[0,0],[55,-0.0008],[57,-0.0014],[60,-0.0026],[62,-0.003],[64,-0.0027],[70,-0.0002],[72,-0.0007],[74,0],[77,-0.0001],[97,-0.0001],[99,-0.0003],[100,-0.0008],[101,-0.0006],[102,-0.0013],[103,-0.0006],[104,-0.0001],[105,0],[106,-0.0009],[107,0],[108,-0.007],[109,-0.0009],[110,-0.006],[111,-0.0013],[112,-0.0031],[113,-0.0001],[114,-0.0001],[115,-0.0003],[133,0],[134,-0.0001],[135,0],[136,-0.001],[137,0],[138,-0.0011],[139,-0.0001],[140,-0.0004],[141,0],[168,0],[171,-0.0007],[172,-0.0001],[180,-0.0001],[182,-0.0005],[183,-0.0012],[184,0]]
       }
       thumb_02_r_ctrl.Location.Z {
        type=MovieSceneScriptingFloatChannel
        default=0
        thinned_from=185
        keys [[0,0],[56,-0.0004],[62,-0.0012],[70,-0.0001],[101,0.0002],[105,0],[107,0],[108,-0.0002],[109,0.0003],[110,0.0022],[111,0.0001],[112,0.0006],[113,-0.0005],[135,0],[136,-0.0007],[137,0],[169,0],[171,-0.0003],[172,0],[181,0.0001],[183,0.0005],[184,0]]
       }
       thumb_02_r_ctrl.Rotation.X {
        type=MovieSceneScriptingFloatChannel
        default=-1.2708
        thinned_from=185
        keys [[0,-1.6319],[1,-1.6163],[2,-1.613],[3,-1.6136],[4,-1.6257],[5,-1.628],[6,-1.5961],[7,-1.5695],[8,-1.5389],[9,-1.552],[10,-1.5931],[11,-1.5875],[12,-1.5758],[13,-1.5854],[14,-1.6028],[15,-1.5809],[16,-1.5744],[17,-1.5277],[18,-1.5407],[19,-1.5352],[20,-1.5217],[21,-1.5342],[22,-1.5076],[25,-1.4987],[26,-1.478],[27,-1.4794],[28,-1.5085],[29,-1.499],[31,-1.5104],[32,-1.5015],[33,-1.5021],[34,-1.531],[35,-1.5085],[36,-1.4818],[37,-1.4953],[38,-1.5332],[39,-1.488],[40,-1.4884],[41,-1.5168],[42,-1.4805],[43,-1.4765],[53,-1.5543],[55,-1.5759],[56,-1.5966],[57,-1.6306],[59,-1.7142],[60,-1.7519],[61,-1.779],[62,-1.7893],[68,-1.7818],[69,-1.7865],[70,-1.7969],[71,-1.9012],[72,-1.9999],[74,-1.9992],[75,-2.0109],[76,-1.9974],[77,-2.0156],[78,-2.0178],[82,-2.0035],[83,-2.133],[84,-2.1502],[85,-2.1533],[87,-2.0994],[88,-2.1004],[89,-2.08],[90,-2.0355],[91,-2.075],[92,-2.0595],[93,-2.0341],[94,-2.0252],[95,-2.0426],[96,-2.0362],[97,-1.8969],[98,-1.8297],[99,-1.7143],[100,-1.4017],[101,-1.7583],[102,-1.7755],[103,-1.0109],[104,-0.6718],[106,-1.0121],[107,-0.8908],[108,-1.0108],[110,-1.0391],[111,-0.7857],[112,-0.7152],[113,-0.7025],[114,-0.6547],[115,-0.4742],[116,-0.4325],[117,-0.4383],[118,-0.4406],[119,-0.4402],[120,-0.4463],[121,-0.46],[122,-0.4764],[123,-0.474],[124,-0.4509],[125,-0.4596],[126,-0.4767],[127,-0.475],[129,-0.47],[130,-0.4772],[131,-0.4708],[132,-0.4831],[134,-0.5319],[135,-0.5316],[136,-0.574],[137,-0.4266],[138,-0.3891],[139,-0.3589],[140,-0.3666],[141,-0.366],[142,-0.3718],[143,-0.3766],[144,-0.3769],[145,-0.3726],[146,-0.3704],[148,-0.3715],[149,-0.3705],[150,-0.3685],[151,-0.3728],[152,-0.3753],[153,-0.3797],[154,-0.3777],[155,-0.3856],[156,-0.3808],[157,-0.3834],[158,-0.3823],[159,-0.3737],[160,-0.3783],[161,-0.3789],[162,-0.385],[163,-0.3845],[164,-0.3853],[165,-0.3797],[166,-0.384],[167,-0.3752],[168,-0.3795],[169,-0.2908],[170,-0.2889],[171,-0.3561],[172,-0.3701],[173,-0.323],[174,-0.3937],[175,-0.351],[176,-0.3918],[177,-0.4062],[178,-0.4106],[179,-0.4137],[180,-0.4811],[181,-0.488],[182,-1.1171],[183,-1.2393],[184,-1.2708]]
       }
       thumb_02_r_ctrl.Rotation.Y {
        type=MovieSceneScriptingFloatChannel
        default=6.8278
        thinned_from=185
        keys [[0,10.2018],[1,10.1988],[2,10.1385],[3,10.1036],[4,10.1318],[5,9.9896],[6,9.8946],[7,9.7336],[8,9.6301],[9,9.5691],[10,9.6895],[11,9.6599],[12,9.7243],[13,9.7393],[14,9.8815],[15,9.8038],[16,9.7674],[17,9.6813],[18,9.6586],[19,9.5818],[20,9.6062],[21,9.5681],[23,9.5319],[24,9.4916],[25,9.4256],[28,9.504],[30,9.3809],[32,9.2837],[33,9.2175],[34,9.3763],[35,9.2072],[36,9.3321],[37,9.382],[38,9.5964],[39,9.4772],[40,9.3874],[41,9.5879],[42,9.4219],[43,9.33],[44,9.3985],[45,9.3817],[46,9.3334],[47,9.2568],[48,9.155],[49,9.0312],[51,8.7299],[55,8.0015],[56,7.6378],[57,7.0047],[58,6.2176],[59,5.3917],[60,4.6426],[61,4.0856],[62,3.8361],[64,3.7257],[65,3.6785],[66,3.6408],[67,3.6159],[68,3.6069],[69,3.6968],[70,3.8947],[71,4.351],[72,4.7013],[74,4.7066],[75,4.4875],[76,4.3082],[77,4.3457],[78,4.112],[80,4.0607],[81,4.0438],[82,4.0374],[83,4.7783],[84,4.9481],[85,4.9405],[86,4.69],[87,4.5446],[88,4.5303],[89,4.3928],[90,4.1765],[91,4.3683],[92,4.4269],[93,4.2219],[94,4.2038],[95,4.3126],[96,4.2314],[97,3.6006],[98,2.9368],[99,2.1673],[100,0.2633],[101,1.8643],[102,5.8396],[103,-2.6221],[104,-2.4923],[105,-1.1601],[106,0.1921],[107,-0.5874],[108,0.0562],[109,0.3099],[110,0.1795],[111,-1.5182],[112,-2.6894],[113,-1.9458],[114,-2.2724],[115,-4.0814],[116,-4.6862],[117,-4.9594],[118,-5.3273],[119,-5.6446],[120,-5.7664],[121,-5.642],[122,-5.5948],[123,-5.7093],[124,-5.3933],[125,-5.2226],[126,-5.2177],[127,-5.1314],[128,-4.9563],[129,-4.9295],[130,-4.7971],[132,-4.9249],[133,-4.785],[134,-5.0071],[135,-4.8474],[136,-3.487],[137,-4.1274],[138,-3.4975],[139,-3.2031],[140,-3.4346],[141,-3.4877],[142,-3.3652],[143,-3.3705],[144,-3.4196],[145,-3.2375],[146,-3.2],[147,-3.2703],[148,-3.2998],[149,-3.302],[150,-3.3238],[151,-3.471],[152,-3.5243],[153,-3.6055],[154,-3.569],[155,-3.5869],[156,-3.5508],[157,-3.5444],[158,-3.4728],[159,-3.4114],[160,-3.4679],[161,-3.5362],[162,-3.5552],[163,-3.5212],[164,-3.5106],[165,-3.4173],[166,-3.2944],[167,-2.6609],[168,-1.0452],[169,0.6808],[170,2.203],[171,3.6173],[172,4.4049],[173,4.4727],[174,5.1036],[175,4.9696],[176,5.3079],[177,5.362],[178,5.3771],[179,5.4523],[180,5.5771],[181,5.0682],[182,5.1839],[183,7.3892],[184,6.8278]]
       }
       thumb_02_r_ctrl.Rotation.Z {
        type=MovieSceneScriptingFloatChannel
        default=23.5546
        thinned_from=185
        keys [[0,24.9796],[1,25.077],[4,24.9386],[5,24.7585],[6,24.858],[7,24.8476],[8,24.9295],[10,24.6403],[11,24.6431],[12,24.7952],[13,24.7489],[15,24.8528],[16,24.8533],[17,25.061],[18,24.9502],[19,24.8986],[20,25.0145],[21,24.8894],[22,25.0537],[25,24.9606],[26,25.1341],[27,25.1417],[28,24.985],[29,24.9682],[31,24.785],[32,24.7795],[33,24.6998],[35,24.6452],[36,24.9652],[38,24.9281],[39,25.0891],[40,24.9845],[42,25.0754],[43,24.9978],[45,24.9763],[46,24.8831],[48,24.5385],[50,24.0238],[53,23.0389],[55,22.3114],[56,21.611],[57,20.3939],[58,18.8826],[59,17.2996],[60,15.8673],[61,14.8083],[62,14.345],[65,14.0942],[67,13.9946],[68,13.9802],[70,14.3819],[72,13.5103],[74,13.5307],[75,12.8658],[76,12.6278],[77,12.4573],[78,11.831],[82,11.8532],[83,11.8376],[84,12.0248],[85,11.9609],[86,11.6796],[87,11.7275],[89,11.6192],[90,11.7279],[91,11.6319],[92,12.021],[93,11.8692],[94,11.9581],[95,11.9782],[96,11.862],[97,12.3467],[98,11.6786],[99,11.5192],[100,11.9108],[101,9.8193],[102,18.4998],[103,11.9859],[104,17.1148],[105,17.0558],[106,17.175],[107,17.3584],[108,16.9222],[109,17.2062],[110,16.8458],[111,17.2076],[112,16.3041],[113,17.546],[114,17.6398],[115,17.295],[117,16.4439],[118,15.8748],[119,15.3814],[120,15.1022],[121,15.0948],[123,14.7457],[124,15.6496],[125,15.7935],[126,15.5447],[127,15.7106],[128,16.031],[130,16.2061],[131,16.1975],[132,15.9226],[133,15.7832],[134,15.0452],[135,15.3264],[136,16.9091],[137,17.7465],[138,18.7352],[139,19.1964],[140,18.8255],[141,18.7399],[142,18.9398],[143,18.9334],[144,18.8548],[145,19.1466],[146,19.2064],[147,19.0929],[148,19.0457],[150,19.0053],[151,18.7699],[152,18.6854],[153,18.5568],[154,18.6147],[155,18.5893],[156,18.6452],[157,18.6568],[159,18.8662],[161,18.6681],[162,18.6401],[164,18.7116],[165,18.8596],[166,19.0587],[167,20.0862],[168,21.8998],[169,24.3646],[170,25.7868],[171,26.619],[172,27.2531],[173,27.6083],[174,27.7035],[175,27.8607],[176,27.9091],[178,27.8618],[179,27.9079],[180,27.6305],[181,27.1225],[182,22.6851],[183,24.4103],[184,23.5546]]
       }
       thumb_02_r_ctrl.Scale.X {type=MovieSceneScriptingFloatChannel, default=1, const [1,185]}
       thumb_02_r_ctrl.Scale.Y {type=MovieSceneScriptingFloatChannel, default=1, const [1,185]}
       thumb_02_r_ctrl.Scale.Z {type=MovieSceneScriptingFloatChannel, default=1, const [1,185]}
       thumb_03_r_ctrl.Location.X {
        type=MovieSceneScriptingFloatChannel
        default=-0.0001
        thinned_from=185
        keys [[0,0],[68,-0.0002],[70,0],[72,-0.0002],[74,0],[99,-0.0001],[100,-0.0003],[101,-0.0003],[102,-0.001],[103,-0.0004],[105,0],[106,-0.0005],[107,0],[108,-0.0048],[109,-0.0008],[110,-0.0043],[111,-0.0004],[112,-0.001],[113,-0.0009],[114,-0.0012],[115,-0.0004],[135,0],[136,-0.0006],[137,0],[138,-0.0006],[139,-0.0001],[140,-0.0003],[141,0],[165,0],[166,-0.0001],[169,-0.0001],[171,-0.0006],[172,-0.0001],[180,0],[183,-0.0006],[184,-0.0001]]
       }
       thumb_03_r_ctrl.Location.Y {
        type=MovieSceneScriptingFloatChannel
        default=0
        thinned_from=185
        keys [[0,0],[64,-0.0006],[68,-0.0003],[70,0],[72,-0.0002],[99,0],[100,-0.0003],[105,0],[107,0],[108,-0.0008],[109,-0.0001],[110,-0.0009],[111,-0.0003],[112,-0.0007],[113,0.0003],[114,0.0004],[115,0.0001],[184,0]]
       }
       thumb_03_r_ctrl.Location.Z {
        type=MovieSceneScriptingFloatChannel
        default=0
        thinned_from=185
        keys [[0,0],[68,-0.0003],[70,0],[101,0],[102,-0.0001],[107,0],[108,-0.0005],[110,0.0005],[111,0],[112,0.0001],[113,-0.0002],[135,0],[136,-0.0003],[137,0],[171,-0.0001],[183,0.0002],[184,0]]
       }
       thumb_03_r_ctrl.Rotation.X {
        type=MovieSceneScriptingFloatChannel
        default=0.4291
        thinned_from=185
        keys [[0,0.4252],[1,0.4379],[2,0.4316],[3,0.4282],[4,0.4107],[5,0.3672],[6,0.4001],[7,0.399],[8,0.4069],[9,0.3813],[10,0.3471],[11,0.3437],[12,0.3465],[13,0.3485],[14,0.3571],[15,0.3769],[16,0.3904],[17,0.425],[18,0.4134],[19,0.4374],[20,0.4565],[21,0.4197],[22,0.4632],[23,0.4452],[24,0.4627],[25,0.4879],[26,0.4857],[27,0.4959],[28,0.4829],[29,0.4537],[30,0.4326],[31,0.427],[32,0.4251],[33,0.397],[34,0.3821],[35,0.3592],[36,0.4392],[37,0.4386],[38,0.4465],[39,0.4453],[40,0.4628],[41,0.4542],[42,0.4805],[43,0.4648],[44,0.4689],[45,0.4645],[46,0.452],[47,0.4322],[48,0.4059],[49,0.3739],[50,0.3369],[51,0.296],[52,0.2517],[53,0.205],[54,0.1567],[55,0.1076],[56,0.0148],[57,-0.1453],[58,-0.3431],[59,-0.5489],[60,-0.7331],[61,-0.866],[62,-0.918],[68,-0.9249],[69,-0.9231],[70,-0.9191],[71,-0.8604],[72,-0.8037],[74,-0.8033],[75,-0.9518],[76,-1.0015],[77,-1.04],[78,-1.1261],[79,-1.1156],[81,-1.0698],[82,-1.0593],[83,-1.1064],[84,-1.1285],[85,-1.1076],[86,-1.1229],[87,-1.1266],[88,-1.1668],[89,-1.1592],[90,-1.1406],[91,-1.1689],[92,-1.0828],[93,-1.144],[94,-1.1139],[95,-1.1413],[96,-1.0741],[97,-0.9432],[98,-0.9153],[99,-0.5776],[100,-0.3191],[101,-0.8075],[102,-0.0348],[103,-0.29],[104,1.795],[105,1.4956],[106,-0.7399],[107,-0.7758],[108,-0.9722],[109,-1.0014],[110,-1.1091],[111,-1.342],[112,-0.9781],[113,-1.0558],[114,-1.2807],[115,-1.3389],[116,-1.4245],[117,-1.4277],[118,-1.4235],[119,-1.5262],[120,-1.5469],[121,-1.5625],[122,-1.5994],[123,-1.6891],[124,-1.6061],[125,-1.6217],[126,-1.6532],[127,-1.6033],[128,-1.6126],[129,-1.5929],[130,-1.591],[131,-1.5788],[132,-1.6479],[133,-1.6193],[134,-1.7942],[135,-1.7702],[136,-1.5013],[137,-1.5811],[138,-1.3453],[139,-1.0668],[140,-1.1477],[142,-1.0833],[143,-1.134],[144,-1.165],[145,-1.1767],[146,-1.1395],[147,-1.166],[148,-1.1756],[149,-1.1505],[150,-1.1201],[151,-1.2002],[152,-1.1808],[153,-1.1924],[154,-1.1727],[155,-1.1905],[156,-1.1955],[157,-1.1976],[158,-1.1843],[159,-1.1371],[160,-1.1746],[161,-1.1816],[162,-1.2198],[163,-1.2071],[164,-1.1294],[165,-0.9259],[166,-0.6809],[167,-0.4085],[168,-0.0165],[169,0.6956],[170,1.0664],[171,1.3871],[172,1.5289],[173,1.5934],[174,1.6364],[175,1.7487],[176,1.7737],[177,1.784],[178,1.7742],[179,1.7333],[180,1.7219],[181,1.509],[182,0.144],[183,0.6436],[184,0.4291]]
       }
       thumb_03_r_ctrl.Rotation.Y {
        type=MovieSceneScriptingFloatChannel
        default=1.1041
        thinned_from=185
        keys [[0,1.0912],[1,1.1214],[2,1.1065],[3,1.0982],[4,1.0562],[5,0.9508],[6,1.0306],[7,1.0279],[8,1.047],[9,0.985],[10,0.9016],[11,0.8931],[13,0.905],[14,0.9258],[15,0.9744],[16,1.0073],[17,1.0906],[18,1.0627],[19,1.1202],[20,1.166],[21,1.078],[22,1.1817],[23,1.1389],[24,1.1807],[25,1.2402],[26,1.235],[27,1.2591],[28,1.2284],[29,1.1592],[30,1.1088],[31,1.0954],[32,1.0908],[33,1.0231],[34,0.987],[35,0.9311],[36,1.1247],[37,1.1231],[38,1.1421],[39,1.1393],[40,1.1809],[41,1.1605],[42,1.2229],[43,1.1855],[44,1.1953],[45,1.1817],[46,1.1428],[47,1.081],[48,0.9989],[49,0.899],[50,0.7839],[51,0.6561],[52,0.5182],[53,0.3726],[54,0.222],[55,0.0688],[56,-0.2208],[57,-0.7208],[58,-1.3387],[59,-1.9819],[60,-2.5582],[61,-2.975],[62,-3.1398],[67,-3.1705],[68,-3.1718],[69,-3.1637],[70,-3.1459],[71,-2.8948],[72,-2.6529],[74,-2.6516],[75,-3.2949],[76,-3.5275],[77,-3.7166],[78,-4.1653],[79,-4.1097],[81,-3.8649],[82,-3.8092],[83,-4.0602],[84,-4.1806],[85,-4.067],[86,-4.15],[87,-4.1699],[88,-4.3974],[89,-4.3537],[90,-4.2481],[91,-4.4082],[92,-3.9353],[93,-4.2676],[94,-4.1008],[95,-4.2524],[96,-3.8895],[97,-3.2534],[98,-3.1275],[99,-1.7919],[100,-0.9238],[101,-2.6575],[102,-0.0906],[103,-0.8415],[104,3.8827],[105,3.3404],[106,-2.3955],[107,-2.5401],[108,-3.3778],[109,-3.5274],[110,-3.5107],[111,-3.2768],[112,-0.5775],[113,0.4065],[114,0.8204],[115,1.3657],[116,1.4244],[117,1.4202],[118,1.4326],[119,1.1985],[120,1.1527],[121,1.1188],[122,1.0341],[123,0.8202],[124,1.0239],[125,0.9893],[126,0.9163],[127,1.039],[128,1.0197],[129,1.0694],[130,1.0766],[131,1.1083],[132,0.9457],[133,1.017],[134,0.5902],[135,0.6533],[136,1.3016],[137,1.1179],[138,1.657],[139,2.258],[140,2.0909],[142,2.2295],[143,2.1252],[144,2.0615],[145,2.0384],[146,2.1185],[147,2.0642],[148,2.0453],[149,2.0997],[150,2.1647],[151,1.997],[152,2.0393],[153,2.0157],[154,2.0585],[155,2.0216],[156,2.0118],[157,2.0079],[158,2.0368],[159,2.1368],[160,2.0583],[161,2.0437],[162,1.9625],[163,1.9896],[164,2.0071],[165,2.0437],[166,1.9904],[167,1.8889],[168,1.4132],[169,1.9941],[170,2.4874],[171,3.1368],[172,3.4013],[173,3.5195],[174,3.5975],[175,3.7978],[176,3.8419],[177,3.8599],[178,3.8428],[179,3.7707],[180,3.7504],[181,3.3681],[182,0.3773],[183,1.5847],[184,1.1041]]
       }
       thumb_03_r_ctrl.Rotation.Z {
        type=MovieSceneScriptingFloatChannel
        default=2.6749
        thinned_from=185
        keys [[0,2.6444],[1,2.7186],[2,2.6819],[3,2.6616],[4,2.5587],[5,2.3008],[6,2.4961],[7,2.4895],[8,2.5363],[9,2.3845],[10,2.1807],[11,2.1603],[13,2.1891],[14,2.2399],[15,2.3585],[16,2.4389],[17,2.643],[18,2.5745],[19,2.7157],[20,2.8279],[21,2.612],[22,2.8664],[23,2.7614],[24,2.8639],[25,3.01],[26,2.9972],[27,3.0565],[28,2.9812],[29,2.8113],[30,2.6875],[31,2.6548],[32,2.6436],[33,2.4778],[34,2.3895],[35,2.2529],[36,2.7267],[37,2.7228],[38,2.7692],[39,2.7625],[40,2.8646],[41,2.8143],[42,2.9676],[43,2.8758],[44,2.8997],[45,2.8676],[46,2.7753],[47,2.629],[48,2.4344],[49,2.1978],[50,1.9251],[51,1.6222],[52,1.2954],[53,0.9504],[54,0.5935],[55,0.2305],[56,-0.4555],[57,-1.6395],[58,-3.1025],[59,-4.6253],[60,-5.989],[61,-6.9745],[62,-7.3628],[67,-7.4266],[68,-7.4293],[69,-7.4108],[70,-7.37],[71,-6.7917],[72,-6.2345],[74,-6.2316],[75,-7.7132],[76,-8.2456],[77,-8.6787],[78,-9.702],[79,-9.5752],[81,-9.017],[82,-8.8901],[83,-9.4622],[84,-9.7364],[85,-9.4778],[86,-9.667],[87,-9.7122],[88,-10.2301],[89,-10.1308],[90,-9.8904],[91,-10.255],[92,-9.1775],[93,-9.9346],[94,-9.5549],[95,-9.8999],[96,-9.0736],[97,-7.6187],[98,-7.3321],[99,-4.2422],[100,-2.1823],[101,-6.2329],[102,-0.2173],[103,-2.0039],[104,9.7193],[105,8.3038],[106,-5.6392],[107,-5.9739],[108,-7.9009],[109,-8.2451],[110,-10.9684],[111,-17.5722],[112,-21.0418],[113,-29.2542],[114,-37.9951],[115,-43.797],[116,-46.4392],[118,-46.4813],[119,-47.0899],[121,-47.3476],[122,-47.5862],[123,-48.1403],[124,-47.6719],[126,-47.9955],[127,-47.7244],[128,-47.801],[129,-47.7079],[131,-47.6691],[132,-48.0955],[133,-47.9481],[134,-49.0164],[135,-48.8894],[136,-47.3261],[137,-47.8046],[138,-46.4954],[139,-45.0155],[140,-45.4601],[142,-45.1554],[143,-45.439],[144,-45.6196],[145,-45.6972],[146,-45.5148],[148,-45.7339],[150,-45.4652],[151,-45.9009],[152,-45.8081],[153,-45.8796],[154,-45.7836],[157,-45.938],[158,-45.8723],[159,-45.6267],[160,-45.8285],[161,-45.8683],[162,-46.0739],[163,-46.0066],[164,-44.2181],[165,-39.4506],[166,-32.8145],[167,-25.0435],[168,-11.1783],[169,1.7643],[170,6.2592],[171,7.7779],[172,8.4619],[173,8.7694],[174,8.9726],[175,9.4973],[176,9.6132],[177,9.6606],[178,9.6155],[179,9.426],[180,9.3729],[181,8.376],[182,0.9114],[183,3.8687],[184,2.6749]]
       }
       thumb_03_r_ctrl.Scale.X {type=MovieSceneScriptingFloatChannel, default=1, const [1,185]}
       thumb_03_r_ctrl.Scale.Y {type=MovieSceneScriptingFloatChannel, default=1, const [1,185]}
       thumb_03_r_ctrl.Scale.Z {type=MovieSceneScriptingFloatChannel, default=1, const [1,185]}
       index_metacarpal_r_ctrl.Location.X {
        type=MovieSceneScriptingFloatChannel
        default=0.0001
        thinned_from=185
        keys [[0,0.0001],[56,-0.0004],[61,-0.0013],[68,-0.0004],[70,0],[72,-0.0002],[74,0.0001],[99,0],[100,-0.0002],[101,-0.0002],[102,-0.0016],[103,-0.0006],[104,0.0001],[105,0.0001],[106,-0.0002],[107,0.0001],[108,-0.0026],[109,-0.0002],[110,-0.0009],[111,-0.0008],[112,-0.0023],[113,-0.0006],[114,-0.001],[115,-0.0004],[118,0],[135,0.0001],[136,-0.0007],[137,0.0001],[138,-0.0003],[139,0.0001],[140,-0.0001],[168,0.0001],[169,-0.0001],[170,-0.0006],[171,-0.0007],[172,0],[180,0.0001],[181,-0.0002],[182,0],[183,-0.0001],[184,0.0001]]
       }
       index_metacarpal_r_ctrl.Location.Y {
        type=MovieSceneScriptingFloatChannel
        default=0.0002
        thinned_from=185
        keys [[0,0.0002],[102,0.0002],[105,0.0002],[106,0.0008],[107,0.0002],[108,0.0039],[109,0.0008],[110,0.0034],[111,0.0001],[112,0.0003],[113,-0.0006],[114,-0.0009],[115,-0.0001],[135,0.0002],[136,0.0003],[137,0.0002],[138,0.0004],[139,0.0002],[169,0.0001],[171,-0.0002],[172,0.0002],[181,0.0002],[183,0.0008],[184,0.0002]]
       }
       index_metacarpal_r_ctrl.Location.Z {
        type=MovieSceneScriptingFloatChannel
        default=0.0001
        thinned_from=185
        keys [[0,0.0001],[46,0.0001],[50,0.0003],[56,0.0012],[60,0.0031],[61,0.0034],[62,0.0035],[63,0.0034],[69,0.0005],[70,0.0002],[72,0.0003],[97,0.0001],[101,0.0002],[102,0.0012],[103,0.0005],[104,0.0002],[105,0.0001],[106,0.0008],[107,0.0001],[108,0.0034],[109,0.0005],[110,0.0027],[111,0.0012],[112,0.003],[113,0.0007],[114,0.0008],[118,0.0002],[133,0.0001],[134,0.0002],[135,0.0001],[136,0.0014],[137,0.0001],[138,0.001],[139,0.0001],[140,0.0004],[141,0.0001],[168,0.0001],[170,0.0007],[171,0.0009],[172,0.0002],[181,0.0001],[183,0.0008],[184,0.0001]]
       }
       index_metacarpal_r_ctrl.Rotation.X {type=MovieSceneScriptingFloatChannel, default=0.01, const [0.01,185]}
       index_metacarpal_r_ctrl.Rotation.Y {type=MovieSceneScriptingFloatChannel, default=-0.4748, const [-0.4748,185]}
       index_metacarpal_r_ctrl.Rotation.Z {type=MovieSceneScriptingFloatChannel, default=-0.74, const [-0.74,185]}
       index_metacarpal_r_ctrl.Scale.X {type=MovieSceneScriptingFloatChannel, default=1, const [1,185]}
       index_metacarpal_r_ctrl.Scale.Y {type=MovieSceneScriptingFloatChannel, default=1, const [1,185]}
       index_metacarpal_r_ctrl.Scale.Z {type=MovieSceneScriptingFloatChannel, default=1, const [1,185]}
       index_01_r_ctrl.Location.X {
        type=MovieSceneScriptingFloatChannel
        default=-0.0001
        thinned_from=185
        keys [[0,0],[51,-0.0004],[56,-0.0011],[60,-0.0031],[61,-0.0034],[62,-0.0035],[64,-0.0032],[68,-0.0014],[69,-0.0007],[70,-0.0002],[72,-0.001],[74,0],[82,-0.0003],[83,0],[97,0],[99,-0.0002],[100,-0.001],[101,-0.0007],[102,-0.0034],[103,-0.0014],[104,-0.0002],[105,0],[106,-0.0027],[107,0],[108,-0.0179],[109,-0.0024],[110,-0.0143],[111,-0.0014],[112,-0.0036],[113,-0.0041],[114,-0.0056],[115,-0.0021],[116,-0.0016],[117,-0.0007],[118,-0.0003],[120,0],[133,0],[134,-0.0002],[135,0],[136,-0.0039],[137,0],[138,-0.0022],[139,-0.0001],[140,-0.0006],[141,0],[165,0],[166,-0.0002],[167,0],[169,-0.0008],[170,-0.0024],[171,-0.0029],[172,-0.0002],[179,0],[180,-0.0001],[181,-0.001],[182,-0.0011],[183,-0.0026],[184,-0.0001]]
       }
       index_01_r_ctrl.Location.Y {
        type=MovieSceneScriptingFloatChannel
        default=0
        thinned_from=185
        keys [[0,0],[48,-0.0002],[55,-0.0013],[56,-0.0016],[60,-0.0043],[61,-0.0047],[62,-0.0049],[63,-0.0047],[65,-0.0036],[69,-0.0006],[70,-0.0002],[72,-0.0004],[74,0],[99,0.0002],[100,-0.0001],[101,0],[102,-0.001],[103,-0.0002],[104,0.0002],[105,0.0001],[106,0.0005],[107,0],[109,0.0002],[110,0.0025],[111,0],[112,0.0006],[113,-0.0013],[114,-0.0013],[115,-0.0003],[133,0],[134,0.0002],[137,0],[138,0.0008],[139,0.0001],[140,0.0003],[141,0],[167,0],[168,-0.0001],[170,0.0001],[171,0],[180,0],[181,-0.0002],[182,0.0006],[183,0.0008],[184,0]]
       }
       index_01_r_ctrl.Location.Z {
        type=MovieSceneScriptingFloatChannel
        default=0
        thinned_from=185
        keys [[0,0],[55,0.0007],[61,0.0024],[63,0.0024],[70,0.0001],[72,0.0005],[74,0],[77,0.0001],[82,-0.0002],[83,0],[97,0.0001],[100,0.0004],[101,0.0003],[102,0.0028],[103,0.0012],[104,-0.0001],[105,0],[106,-0.0003],[107,0],[108,-0.0023],[109,-0.0006],[110,-0.0021],[111,0.002],[112,0.0048],[113,0.0002],[114,-0.0001],[115,0.0004],[135,0],[136,0.0003],[137,0],[138,0.0002],[139,0],[140,0.0002],[141,0],[168,0],[169,0.0002],[170,0.0007],[171,0.0009],[172,0.0001],[180,0.0001],[181,0.0004],[182,-0.0003],[183,-0.0004],[184,0]]
       }
       index_01_r_ctrl.Rotation.X {
        type=MovieSceneScriptingFloatChannel
        default=10.3866
        thinned_from=185
        keys [[0,10.9246],[1,10.9243],[2,10.9014],[3,10.9238],[5,10.9153],[7,10.9467],[8,10.8997],[10,10.8926],[11,10.9116],[12,10.8899],[15,10.9404],[17,11.0077],[20,11.0057],[26,11.0402],[28,10.9993],[31,11.0205],[32,10.9898],[33,11.0059],[34,10.9837],[35,11.0123],[36,11.0042],[40,11.0274],[41,11.0092],[44,11.0731],[45,11.0568],[46,11.01],[47,10.9358],[49,10.7172],[51,10.4254],[55,9.7197],[56,9.3637],[57,8.7397],[58,7.9602],[59,7.1374],[60,6.3838],[61,5.8116],[62,5.5334],[64,5.3527],[65,5.2754],[66,5.2137],[67,5.1729],[68,5.1581],[69,5.2003],[70,5.2931],[71,5.7223],[72,6.1008],[73,6.0485],[75,5.8512],[77,5.7002],[78,5.6715],[82,5.6387],[83,5.5932],[84,5.593],[86,5.496],[88,5.5464],[90,5.4811],[93,5.5035],[94,5.5355],[95,5.4052],[96,5.1627],[97,4.9586],[98,4.6537],[99,4.0569],[100,3.5207],[101,2.3696],[102,2.8402],[103,3.6984],[104,7.262],[105,7.1189],[106,2.437],[107,2.2843],[108,2.048],[109,2.0791],[110,1.8195],[111,1.5298],[112,2.2104],[113,2.275],[114,2.3041],[115,2.0023],[116,1.6805],[117,1.2886],[118,0.9923],[119,0.6177],[120,0.4259],[121,0.2623],[122,-0.0147],[123,-0.2198],[124,-0.3924],[125,-0.2994],[126,-0.3113],[127,-0.3623],[128,-0.3619],[129,-0.2734],[130,-0.3453],[131,-0.414],[132,-0.4735],[133,-0.3934],[134,-0.273],[135,-0.2513],[136,-0.5196],[137,-1.0413],[138,-1.1447],[139,-0.979],[140,-1.0667],[141,-0.9471],[142,-0.857],[143,-0.8837],[144,-0.8213],[145,-0.8325],[146,-0.8181],[147,-0.7917],[148,-0.767],[149,-0.7019],[150,-0.6898],[151,-0.6895],[152,-0.6099],[153,-0.4916],[154,-0.3492],[155,-0.3728],[156,-0.3832],[157,-0.3948],[158,-0.3926],[159,-0.3884],[160,-0.3566],[161,-0.3519],[162,-0.4063],[163,-0.4147],[164,-0.4069],[165,-0.414],[166,-0.2685],[167,0.9167],[168,3.9456],[169,7.1861],[170,9.2195],[171,10.5759],[172,10.9029],[173,11.1159],[175,11.0998],[176,11.0295],[177,11.036],[179,11.0847],[180,11.086],[181,11.0157],[182,9.6364],[183,9.8452],[184,10.3866]]
       }
       index_01_r_ctrl.Rotation.Y {
        type=MovieSceneScriptingFloatChannel
        default=-1.5208
        thinned_from=185
        keys [[0,-1.4493],[1,-1.4348],[2,-1.4638],[3,-1.4493],[4,-1.443],[5,-1.5003],[6,-1.4567],[7,-1.4639],[8,-1.4184],[9,-1.4363],[10,-1.4186],[11,-1.4307],[12,-1.3598],[13,-1.4146],[14,-1.4514],[15,-1.4333],[16,-1.5005],[17,-1.4321],[18,-1.4206],[19,-1.4462],[20,-1.4298],[21,-1.4916],[22,-1.4516],[23,-1.5285],[24,-1.4722],[25,-1.451],[26,-1.4892],[27,-1.5062],[28,-1.5115],[29,-1.458],[30,-1.4868],[31,-1.5073],[32,-1.5112],[33,-1.5183],[34,-1.4426],[35,-1.4666],[36,-1.429],[37,-1.4612],[38,-1.4713],[39,-1.5033],[40,-1.5304],[41,-1.4946],[42,-1.469],[43,-1.4793],[44,-1.5523],[46,-1.5479],[48,-1.5358],[52,-1.4956],[55,-1.458],[56,-1.4316],[57,-1.3838],[58,-1.3225],[59,-1.2557],[60,-1.1916],[61,-1.1382],[62,-1.1036],[64,-1.0602],[65,-1.0416],[66,-1.0268],[67,-1.017],[68,-1.0135],[69,-1.0848],[70,-1.2327],[71,-1.4672],[72,-1.6253],[73,-1.6119],[74,-1.5823],[75,-1.3733],[76,-1.2929],[77,-1.0726],[78,-0.9458],[79,-0.9491],[81,-0.964],[82,-0.9673],[83,-0.6642],[84,-0.5526],[85,-0.339],[86,-0.3357],[87,-0.2419],[88,-0.3568],[89,-0.2622],[90,-0.2098],[91,-0.1907],[92,-0.0194],[93,-0.0006],[94,-0.1936],[95,0.0451],[96,0.2149],[97,-0.4014],[98,-0.7814],[99,-0.7087],[100,-0.9461],[101,0.0361],[102,1.1043],[103,0.0484],[104,-2.3015],[105,-2.3989],[106,-1.4567],[107,-1.6174],[108,-1.4563],[109,-1.6519],[110,-1.5121],[111,-1.1404],[112,-0.4588],[113,-0.9028],[114,-1.355],[115,-2.0345],[116,-2.1784],[117,-2.2621],[118,-2.388],[119,-2.3653],[120,-2.4728],[121,-2.4719],[122,-2.4096],[123,-2.3344],[124,-1.6248],[125,-1.6202],[126,-1.7643],[127,-1.6737],[128,-1.367],[129,-1.3929],[130,-1.2942],[131,-1.2622],[132,-1.2778],[133,-1.4957],[134,-2.0239],[135,-2.0718],[136,-1.7833],[137,-1.2365],[138,-0.9864],[139,-0.9214],[140,-0.9119],[141,-0.9148],[142,-0.9459],[143,-0.9678],[144,-1.0306],[145,-0.9907],[146,-0.9268],[147,-0.9544],[148,-0.9714],[149,-0.9707],[150,-1.0391],[151,-0.9921],[152,-1.0804],[153,-1.2006],[154,-1.2123],[155,-1.2004],[156,-1.1767],[158,-1.185],[159,-1.1004],[160,-1.0895],[161,-1.106],[162,-1.0792],[163,-1.1596],[164,-1.1381],[165,-1.1128],[166,-1.1025],[167,-1.2934],[168,-1.3546],[169,-1.3521],[170,-1.1163],[171,-1.789],[172,-1.8756],[173,-1.9513],[174,-2.0315],[175,-1.8624],[176,-2.0804],[177,-1.9636],[178,-1.893],[179,-1.9688],[180,-1.8697],[181,-1.702],[182,-1.3645],[183,-1.6501],[184,-1.5208]]
       }
       index_01_r_ctrl.Rotation.Z {
        type=MovieSceneScriptingFloatChannel
        default=-62.4264
        thinned_from=185
        keys [[0,-63.9754],[2,-63.7887],[3,-63.9085],[4,-63.7847],[7,-63.9981],[8,-63.6988],[9,-63.7399],[10,-63.5863],[11,-63.7223],[12,-63.52],[15,-63.8538],[16,-64.0853],[19,-64.3572],[20,-64.2595],[21,-64.4459],[24,-64.5147],[25,-64.699],[27,-64.4544],[28,-64.4473],[29,-64.2073],[30,-64.2452],[31,-64.43],[32,-64.2362],[33,-64.3605],[34,-64.1105],[36,-64.2586],[37,-64.2028],[38,-64.3874],[39,-64.2871],[40,-64.4576],[41,-64.3231],[43,-64.5248],[44,-64.7728],[45,-64.6965],[47,-64.1297],[49,-63.1055],[51,-61.7384],[55,-58.4325],[56,-56.7736],[57,-53.876],[58,-50.2647],[59,-46.4647],[60,-43.0011],[61,-40.399],[62,-39.1833],[65,-38.2276],[66,-37.9992],[67,-37.848],[68,-37.7933],[69,-37.9742],[70,-38.237],[71,-38.3873],[72,-38.4556],[73,-38.202],[76,-36.9394],[78,-35.5362],[82,-35.4984],[83,-35.8308],[84,-35.9528],[85,-35.7173],[86,-35.649],[88,-35.9143],[92,-35.2596],[93,-35.3787],[94,-35.6379],[95,-34.9755],[96,-34.2416],[97,-33.3631],[98,-32.1269],[99,-29.8838],[100,-30.9406],[101,-26.9114],[102,-28.7674],[103,-31.5199],[104,-49.6034],[105,-48.7402],[106,-27.421],[107,-27.3138],[108,-25.8574],[109,-26.0318],[110,-24.6895],[111,-23.5117],[112,-28.6435],[113,-29.5041],[114,-29.6058],[117,-26.586],[119,-24.4134],[120,-23.9048],[121,-23.4902],[122,-22.565],[123,-22.1283],[124,-22.9543],[125,-23.2303],[126,-22.9526],[127,-22.8642],[128,-23.3535],[129,-23.6765],[130,-23.4197],[131,-23.3293],[132,-23.047],[133,-22.9252],[134,-22.5893],[135,-22.4934],[136,-21.7663],[137,-20.1521],[138,-20.1145],[139,-20.859],[140,-20.6353],[141,-21.0472],[142,-21.4169],[143,-21.3737],[144,-21.4209],[145,-21.3969],[147,-21.5481],[148,-21.5885],[149,-21.9118],[150,-21.908],[151,-21.9557],[152,-22.1941],[153,-22.615],[154,-23.1859],[155,-23.0002],[156,-23.0273],[157,-22.9596],[158,-22.9424],[160,-23.1066],[161,-23.1524],[163,-22.9127],[164,-22.9607],[165,-22.9218],[166,-23.3768],[167,-26.1265],[168,-34.2916],[169,-47.3213],[170,-56.2578],[171,-64.7791],[172,-67.2273],[173,-68.9538],[174,-69.2996],[176,-68.8538],[178,-68.6438],[179,-68.6647],[180,-68.5006],[181,-67.7664],[182,-58.9031],[183,-60.1617],[184,-62.4264]]
       }
       index_01_r_ctrl.Scale.X {type=MovieSceneScriptingFloatChannel, default=1, const [1,185]}
       index_01_r_ctrl.Scale.Y {type=MovieSceneScriptingFloatChannel, default=1, const [1,185]}
       index_01_r_ctrl.Scale.Z {type=MovieSceneScriptingFloatChannel, default=1, const [1,185]}
       index_02_r_ctrl.Location.X {
        type=MovieSceneScriptingFloatChannel
        default=-0.0001
        thinned_from=185
        keys [[0,-0.0001],[65,-0.0005],[70,-0.0001],[82,-0.0002],[97,-0.0002],[99,-0.0005],[100,-0.0011],[101,-0.0009],[102,-0.0012],[103,-0.0006],[105,-0.0001],[106,-0.0008],[107,-0.0001],[108,-0.0045],[109,-0.001],[110,-0.0059],[111,-0.0008],[112,-0.0018],[113,-0.001],[114,-0.0014],[115,-0.0007],[116,-0.0005],[117,-0.0002],[133,-0.0001],[134,-0.0002],[135,-0.0001],[136,-0.001],[137,-0.0001],[138,-0.001],[139,-0.0001],[140,-0.0004],[141,-0.0001],[165,-0.0001],[166,-0.0005],[167,-0.0001],[168,-0.0006],[169,-0.0008],[170,-0.0013],[171,-0.0014],[172,-0.0002],[180,-0.0001],[181,-0.0006],[182,-0.0006],[184,-0.0001]]
       }
       index_02_r_ctrl.Location.Y {
        type=MovieSceneScriptingFloatChannel
        default=-0.0001
        thinned_from=185
        keys [[0,-0.0001],[63,0.0004],[68,-0.0002],[99,-0.0003],[100,-0.0008],[103,-0.0002],[104,-0.0003],[105,-0.0001],[106,-0.001],[107,-0.0001],[108,-0.0054],[109,-0.001],[110,-0.0044],[111,-0.0001],[112,0],[113,-0.0012],[114,-0.0014],[115,-0.0008],[118,-0.0002],[135,-0.0001],[136,-0.0016],[137,-0.0001],[138,-0.0008],[139,-0.0002],[140,-0.0003],[165,-0.0001],[166,-0.0004],[167,-0.0001],[169,-0.0007],[170,-0.0011],[171,-0.0013],[172,-0.0002],[180,-0.0001],[182,-0.0008],[183,-0.0007],[184,-0.0001]]
       }
       index_02_r_ctrl.Location.Z {
        type=MovieSceneScriptingFloatChannel
        default=0
        thinned_from=185
        keys [[0,-0.0001],[68,0.0003],[70,0],[99,0],[100,0.0002],[101,0.0001],[102,0.0004],[104,0],[106,0.0001],[107,-0.0001],[108,0.001],[109,0],[110,0.0011],[111,0.0001],[112,0.0002],[113,-0.0007],[114,-0.001],[115,-0.0003],[136,0],[168,-0.0001],[169,-0.0003],[170,-0.0007],[171,-0.0009],[172,-0.0001],[182,0.0001],[183,-0.0001],[184,0]]
       }
       index_02_r_ctrl.Rotation.X {
        type=MovieSceneScriptingFloatChannel
        default=-2.4461
        thinned_from=185
        keys [[0,-2.7063],[2,-2.7111],[3,-2.7047],[4,-2.7107],[6,-2.7051],[8,-2.7102],[9,-2.7059],[11,-2.7069],[12,-2.7144],[15,-2.7092],[16,-2.7016],[20,-2.6991],[24,-2.6894],[25,-2.6828],[27,-2.6842],[28,-2.6799],[29,-2.686],[30,-2.6858],[31,-2.6773],[32,-2.6819],[33,-2.6803],[34,-2.6859],[37,-2.6874],[38,-2.6825],[42,-2.6857],[47,-2.656],[53,-2.5863],[55,-2.5549],[56,-2.5237],[57,-2.4704],[59,-2.3355],[60,-2.2734],[61,-2.2275],[62,-2.2076],[68,-2.1928],[70,-2.1713],[71,-2.1553],[72,-2.1454],[73,-2.1509],[74,-2.1631],[75,-2.1851],[76,-2.1991],[77,-2.2522],[78,-2.2773],[82,-2.2814],[83,-2.3032],[84,-2.3395],[85,-2.3625],[89,-2.3995],[92,-2.4147],[93,-2.4233],[94,-2.4081],[95,-2.4191],[96,-2.382],[97,-2.2487],[98,-2.0899],[99,-1.9151],[100,-1.6948],[101,-2.0408],[102,-2.3066],[103,-2.1061],[104,-0.8984],[105,-0.7834],[106,-2.265],[107,-2.2498],[108,-2.2836],[109,-2.283],[110,-2.2716],[111,-2.2987],[112,-1.3518],[113,-1.0383],[114,-0.9327],[115,-0.833],[117,-0.6867],[118,-0.6301],[119,-0.5834],[120,-0.5429],[121,-0.5237],[122,-0.4854],[123,-0.4611],[124,-0.5035],[125,-0.5231],[126,-0.5051],[127,-0.4996],[128,-0.5253],[129,-0.5294],[130,-0.521],[131,-0.5172],[132,-0.5149],[133,-0.5059],[134,-0.4835],[135,-0.4736],[136,-0.4427],[137,-0.4248],[138,-0.4095],[139,-0.4123],[140,-0.4034],[141,-0.4026],[142,-0.4117],[143,-0.4147],[144,-0.4159],[145,-0.4202],[146,-0.4281],[147,-0.4286],[149,-0.4342],[150,-0.4317],[151,-0.4429],[152,-0.4409],[153,-0.4458],[155,-0.4531],[156,-0.4513],[157,-0.4564],[158,-0.4557],[159,-0.4649],[160,-0.4692],[161,-0.4704],[162,-0.4661],[163,-0.4632],[164,-0.4558],[165,-0.463],[166,-0.4868],[167,-0.6866],[168,-1.222],[169,-1.6888],[170,-2.064],[171,-2.1907],[172,-2.2263],[173,-2.2463],[175,-2.2493],[176,-2.2474],[177,-2.2516],[178,-2.2796],[179,-2.2912],[181,-2.3508],[182,-2.3934],[183,-2.3855],[184,-2.4461]]
       }
       index_02_r_ctrl.Rotation.Y {
        type=MovieSceneScriptingFloatChannel
        default=-0.6556
        thinned_from=185
        keys [[0,-0.9732],[2,-0.9806],[3,-0.9709],[4,-0.98],[5,-0.9728],[6,-0.9714],[7,-0.9732],[8,-0.9792],[9,-0.9726],[10,-0.9745],[11,-0.9742],[12,-0.9857],[13,-0.9811],[15,-0.9776],[16,-0.9662],[17,-0.967],[18,-0.9629],[19,-0.9616],[20,-0.9625],[21,-0.9556],[22,-0.9564],[24,-0.9481],[25,-0.9384],[27,-0.9404],[28,-0.9342],[29,-0.9431],[30,-0.9427],[31,-0.9306],[32,-0.9371],[33,-0.9349],[34,-0.9429],[35,-0.94],[36,-0.9416],[37,-0.9451],[38,-0.938],[39,-0.9407],[40,-0.9363],[41,-0.9405],[42,-0.9427],[44,-0.932],[49,-0.8842],[51,-0.861],[53,-0.8336],[55,-0.8009],[56,-0.7683],[57,-0.7132],[58,-0.6454],[59,-0.575],[60,-0.512],[61,-0.4663],[62,-0.448],[68,-0.4426],[69,-0.4359],[70,-0.4253],[71,-0.4135],[72,-0.4062],[73,-0.4104],[74,-0.4192],[75,-0.4345],[76,-0.4447],[77,-0.4845],[78,-0.504],[82,-0.5072],[83,-0.5251],[84,-0.5557],[85,-0.5758],[86,-0.5826],[87,-0.591],[88,-0.6011],[89,-0.6097],[90,-0.6138],[91,-0.6213],[92,-0.6241],[93,-0.6325],[94,-0.6177],[95,-0.6283],[96,-0.5929],[97,-0.4805],[98,-0.3703],[99,-0.2724],[100,-0.1848],[101,-0.3461],[102,-0.5205],[103,-0.3792],[104,-0.0613],[105,-0.0729],[106,-0.4929],[107,-0.4825],[108,-0.509],[109,-0.5082],[110,-0.4979],[111,-0.4954],[112,-0.0664],[113,-0.0653],[115,-0.0709],[116,-0.078],[117,-0.0883],[118,-0.0971],[119,-0.1065],[120,-0.1151],[121,-0.12],[122,-0.1294],[123,-0.1365],[124,-0.1248],[125,-0.1199],[126,-0.1243],[127,-0.1257],[128,-0.1192],[129,-0.1183],[130,-0.1203],[131,-0.1212],[132,-0.1219],[133,-0.1241],[134,-0.1299],[135,-0.1326],[137,-0.1467],[138,-0.1515],[139,-0.1505],[140,-0.1531],[141,-0.1535],[142,-0.1507],[143,-0.1498],[144,-0.1494],[145,-0.1481],[146,-0.1457],[147,-0.1456],[149,-0.1439],[150,-0.1446],[151,-0.1413],[152,-0.1419],[153,-0.1405],[155,-0.1383],[156,-0.1389],[157,-0.1374],[158,-0.1377],[159,-0.135],[160,-0.1339],[161,-0.1335],[162,-0.1347],[163,-0.1356],[164,-0.1376],[165,-0.1356],[166,-0.129],[167,-0.0865],[168,-0.0788],[169,-0.182],[170,-0.362],[171,-0.4417],[172,-0.4648],[173,-0.4799],[174,-0.4797],[175,-0.4821],[176,-0.4806],[177,-0.4839],[178,-0.5057],[179,-0.5151],[180,-0.5416],[181,-0.5658],[182,-0.603],[183,-0.5949],[184,-0.6556]]
       }
       index_02_r_ctrl.Rotation.Z {
        type=MovieSceneScriptingFloatChannel
        default=-34.7774
        thinned_from=185
        keys [[0,-46.3714],[2,-46.6202],[3,-46.2878],[4,-46.5997],[5,-46.3587],[6,-46.3085],[7,-46.3724],[8,-46.575],[9,-46.3509],[11,-46.4055],[12,-46.7918],[13,-46.6372],[15,-46.5212],[16,-46.1344],[17,-46.1604],[18,-46.0209],[20,-46.0075],[21,-45.7741],[22,-45.8039],[24,-45.5205],[25,-45.1875],[27,-45.2575],[28,-45.0436],[29,-45.348],[30,-45.3364],[31,-44.919],[32,-45.144],[33,-45.0677],[34,-45.3418],[35,-45.2442],[37,-45.4164],[38,-45.1751],[39,-45.2665],[40,-45.116],[42,-45.3335],[44,-44.9698],[48,-43.5801],[51,-42.2342],[53,-41.1374],[55,-39.847],[56,-38.5674],[57,-36.3889],[58,-33.7058],[59,-30.9124],[60,-28.4034],[61,-26.5729],[62,-25.8155],[68,-25.4576],[69,-25.1441],[70,-24.6609],[71,-24.0948],[72,-23.7431],[73,-23.9415],[74,-24.369],[75,-25.1176],[76,-25.606],[77,-27.4472],[78,-28.3808],[82,-28.5294],[83,-29.3248],[84,-30.665],[85,-31.5217],[86,-31.808],[87,-32.1614],[88,-32.5828],[89,-32.935],[90,-33.1031],[91,-33.4136],[92,-33.5241],[93,-33.8638],[94,-33.2661],[95,-33.6985],[96,-32.2249],[97,-27.2695],[98,-21.7823],[99,-16.0969],[100,-9.6664],[101,-20.4292],[102,-28.9205],[103,-22.244],[104,13.2546],[105,16.5825],[106,-27.9456],[107,-27.3954],[108,-28.6087],[109,-28.5868],[110,-28.1951],[111,-27.5627],[112,3.6712],[113,9.4402],[114,12.4374],[115,15.2509],[117,19.3843],[118,20.9833],[119,22.3315],[120,23.4874],[121,24.0461],[122,25.1547],[123,25.8733],[124,24.6271],[125,24.0571],[126,24.5791],[127,24.7401],[128,23.9919],[129,23.8752],[130,24.1181],[132,24.294],[133,24.557],[134,25.1994],[135,25.4965],[136,26.3488],[137,26.9299],[138,27.3887],[139,27.2978],[140,27.5489],[141,27.5869],[142,27.3185],[143,27.2283],[144,27.1926],[145,27.0648],[146,26.8325],[147,26.8188],[149,26.6534],[150,26.7288],[151,26.3992],[152,26.4562],[155,26.0981],[156,26.1517],[157,26.0013],[158,26.0231],[159,25.7539],[160,25.6279],[161,25.5917],[163,25.8046],[164,26.0188],[165,25.808],[166,25.1536],[167,19.3523],[168,4.2234],[169,-9.3274],[170,-21.079],[171,-25.3666],[172,-26.5646],[173,-27.2739],[174,-27.266],[175,-27.3783],[176,-27.3083],[177,-27.4582],[178,-28.459],[179,-28.8804],[180,-30.048],[181,-31.0895],[182,-32.6974],[183,-32.3807],[184,-34.7774]]
       }
       index_02_r_ctrl.Scale.X {type=MovieSceneScriptingFloatChannel, default=1, const [1,185]}
       index_02_r_ctrl.Scale.Y {type=MovieSceneScriptingFloatChannel, default=1, const [1,185]}
       index_02_r_ctrl.Scale.Z {type=MovieSceneScriptingFloatChannel, default=1, const [1,185]}
       index_03_r_ctrl.Location.X {
        type=MovieSceneScriptingFloatChannel
        default=-0.0001
        thinned_from=185
        keys [[0,-0.0001],[55,-0.0007],[61,-0.0025],[63,-0.0026],[69,-0.0004],[70,-0.0002],[72,-0.0004],[74,-0.0001],[95,-0.0001],[97,-0.0008],[99,-0.0012],[100,-0.0002],[101,-0.0001],[102,-0.0024],[103,-0.0011],[104,-0.0005],[105,-0.0001],[106,-0.0003],[107,-0.0001],[108,-0.0018],[109,-0.0005],[110,-0.0026],[111,-0.0062],[112,-0.0177],[113,-0.0012],[114,-0.0017],[115,-0.0007],[116,-0.0006],[117,-0.0002],[133,-0.0001],[134,-0.0002],[135,-0.0001],[136,-0.0007],[137,-0.0001],[138,-0.0007],[139,-0.0001],[140,-0.0003],[141,-0.0001],[165,-0.0001],[166,-0.002],[167,-0.0001],[168,-0.0018],[169,-0.0019],[170,-0.0025],[171,-0.0019],[172,-0.0002],[180,-0.0001],[181,-0.0004],[184,-0.0001]]
       }
       index_03_r_ctrl.Location.Y {
        type=MovieSceneScriptingFloatChannel
        default=0.0001
        thinned_from=185
        keys [[0,0],[63,-0.0005],[95,0],[99,-0.0005],[100,0],[101,0],[102,-0.0016],[103,-0.0005],[105,0],[106,-0.0002],[107,0.0001],[108,-0.0011],[109,-0.0003],[110,-0.0015],[111,-0.0046],[112,-0.0064],[113,-0.0008],[114,-0.0009],[118,0],[135,0],[136,-0.0011],[137,0],[138,-0.0005],[139,0],[140,-0.0001],[165,0.0001],[166,-0.0008],[167,0.0001],[168,-0.0006],[170,-0.0012],[171,-0.0008],[172,0],[180,0],[181,-0.0001],[183,0.0002],[184,0.0001]]
       }
       index_03_r_ctrl.Location.Z {
        type=MovieSceneScriptingFloatChannel
        default=-0.0001
        thinned_from=185
        keys [[0,-0.0001],[99,-0.0003],[101,-0.0001],[102,-0.0004],[104,0.0001],[105,-0.0001],[106,0.0004],[107,-0.0001],[108,0.0027],[109,0.0004],[110,0.0029],[111,-0.0001],[112,-0.0019],[113,-0.0009],[114,-0.0012],[115,-0.0005],[135,-0.0001],[136,0.0001],[137,-0.0001],[167,-0.0001],[169,-0.0005],[170,-0.0009],[171,-0.0008],[172,-0.0001],[181,-0.0001],[182,0.0003],[183,0],[184,-0.0001]]
       }
       index_03_r_ctrl.Rotation.X {
        type=MovieSceneScriptingFloatChannel
        default=-0.6584
        thinned_from=185
        keys [[0,-0.8151],[1,-0.8174],[2,-0.8125],[3,-0.817],[4,-0.8183],[6,-0.8135],[7,-0.8192],[9,-0.8173],[10,-0.8213],[11,-0.8199],[12,-0.826],[13,-0.8232],[14,-0.8249],[16,-0.8181],[20,-0.8138],[22,-0.8099],[24,-0.8028],[26,-0.8018],[27,-0.7986],[28,-0.8006],[29,-0.8044],[30,-0.8062],[31,-0.8029],[32,-0.8052],[33,-0.8037],[34,-0.8071],[35,-0.8065],[36,-0.8078],[40,-0.8036],[41,-0.8073],[43,-0.8051],[46,-0.7938],[49,-0.7772],[52,-0.7557],[55,-0.7286],[56,-0.7101],[57,-0.6787],[58,-0.64],[59,-0.5999],[60,-0.5639],[61,-0.5379],[62,-0.5276],[68,-0.5253],[69,-0.5224],[70,-0.5166],[71,-0.4982],[72,-0.4829],[73,-0.4847],[74,-0.4885],[76,-0.5049],[77,-0.5297],[78,-0.545],[83,-0.5474],[84,-0.5553],[85,-0.5618],[86,-0.5717],[87,-0.5714],[88,-0.5818],[90,-0.5842],[91,-0.5971],[92,-0.5874],[93,-0.5914],[94,-0.5829],[95,-0.5855],[96,-0.5546],[97,-0.4945],[98,-0.4234],[99,-0.3492],[100,-0.2682],[101,-0.5293],[102,-0.9562],[103,-0.3643],[104,-0.3027],[105,-0.325],[106,-0.428],[107,-0.4304],[108,-0.4405],[109,-0.4351],[110,-0.4232],[111,-0.4198],[112,-0.1639],[113,-0.0493],[114,-0.0122],[115,0.026],[116,0.0501],[117,0.073],[118,0.0915],[119,0.1115],[120,0.1206],[121,0.1286],[122,0.1405],[123,0.1468],[124,0.1583],[125,0.1537],[126,0.1541],[127,0.1553],[128,0.1519],[129,0.1511],[130,0.1528],[131,0.1556],[132,0.1557],[133,0.1572],[134,0.1533],[135,0.1546],[136,0.1636],[137,0.1703],[138,0.1809],[139,0.1836],[140,0.1853],[142,0.184],[143,0.1852],[144,0.1827],[145,0.1831],[146,0.1812],[147,0.1815],[148,0.1796],[149,0.18],[150,0.182],[151,0.1778],[152,0.1774],[153,0.1762],[154,0.1768],[155,0.1745],[156,0.177],[157,0.1737],[158,0.174],[159,0.1707],[160,0.1683],[161,0.1692],[162,0.1707],[163,0.1728],[164,0.1734],[165,0.173],[166,0.1661],[167,0.0948],[168,-0.0929],[169,-0.2583],[170,-0.4219],[171,-0.5238],[172,-0.5512],[173,-0.5733],[174,-0.5721],[175,-0.5689],[176,-0.5697],[177,-0.5808],[178,-0.5966],[179,-0.5943],[180,-0.6103],[181,-0.6157],[182,-0.6184],[183,-0.6236],[184,-0.6584]]
       }
       index_03_r_ctrl.Rotation.Y {
        type=MovieSceneScriptingFloatChannel
        default=-0.4599
        thinned_from=185
        keys [[0,-0.4278],[1,-0.4274],[2,-0.4282],[4,-0.4273],[6,-0.4281],[7,-0.4272],[9,-0.4274],[11,-0.427],[12,-0.4261],[27,-0.4304],[30,-0.4292],[31,-0.4297],[40,-0.4296],[44,-0.4298],[49,-0.4357],[55,-0.4487],[56,-0.4535],[57,-0.4616],[59,-0.482],[60,-0.4913],[61,-0.4979],[62,-0.5005],[68,-0.4999],[69,-0.5009],[70,-0.503],[71,-0.5098],[72,-0.5154],[74,-0.5133],[76,-0.5074],[77,-0.4991],[78,-0.4933],[83,-0.4925],[85,-0.4877],[86,-0.4846],[87,-0.4847],[88,-0.4815],[90,-0.4807],[91,-0.4769],[92,-0.4798],[93,-0.4786],[94,-0.4811],[95,-0.4803],[96,-0.4905],[97,-0.5121],[98,-0.541],[99,-0.5753],[100,-0.6132],[101,-0.4979],[102,-0.4472],[103,-0.5837],[104,-0.5973],[105,-0.585],[106,-0.5375],[107,-0.5363],[108,-0.5309],[110,-0.5369],[111,-0.5519],[112,-0.7109],[113,-0.7626],[115,-0.8242],[117,-0.8665],[118,-0.8836],[119,-0.9034],[121,-0.9209],[122,-0.9331],[123,-0.9396],[124,-0.9525],[125,-0.9475],[127,-0.9492],[128,-0.9455],[129,-0.9446],[131,-0.9495],[133,-0.9513],[134,-0.9469],[135,-0.9485],[136,-0.956],[138,-0.9774],[140,-0.983],[142,-0.9816],[143,-0.9829],[144,-0.9801],[145,-0.9805],[146,-0.9783],[147,-0.9786],[148,-0.9765],[150,-0.9792],[151,-0.9744],[153,-0.9726],[154,-0.9732],[155,-0.9707],[156,-0.9735],[157,-0.9698],[158,-0.9701],[160,-0.9637],[164,-0.9694],[165,-0.969],[166,-0.9617],[167,-0.8861],[168,-0.7272],[169,-0.6225],[170,-0.5428],[171,-0.5013],[172,-0.4913],[173,-0.4842],[175,-0.4856],[176,-0.4852],[177,-0.4818],[178,-0.477],[179,-0.4777],[180,-0.473],[183,-0.4689],[184,-0.4599]]
       }
       index_03_r_ctrl.Rotation.Z {
        type=MovieSceneScriptingFloatChannel
        default=5.3056
        thinned_from=185
        keys [[0,-0.2263],[1,-0.3077],[2,-0.136],[3,-0.2955],[4,-0.339],[5,-0.2458],[6,-0.1699],[7,-0.3712],[8,-0.3126],[9,-0.3048],[10,-0.443],[11,-0.3958],[12,-0.6077],[13,-0.5115],[14,-0.5701],[15,-0.4436],[16,-0.3322],[17,-0.2846],[18,-0.2316],[19,-0.2235],[20,-0.1821],[21,-0.1077],[22,-0.0463],[23,0.0766],[24,0.2027],[25,0.2269],[26,0.2406],[27,0.3504],[28,0.2811],[29,0.1469],[30,0.0832],[31,0.2006],[32,0.1212],[33,0.1728],[34,0.0545],[35,0.0757],[36,0.027],[37,0.0776],[38,0.1086],[39,0.153],[40,0.1775],[41,0.0454],[42,0.0763],[43,0.1235],[44,0.215],[45,0.3608],[46,0.5243],[47,0.7061],[48,0.9071],[49,1.1279],[50,1.3693],[51,1.632],[52,1.9167],[53,2.2242],[54,2.5552],[55,2.9104],[56,3.5824],[57,4.7243],[58,6.127],[59,7.5816],[60,8.8793],[61,9.8111],[62,10.1682],[68,10.0915],[69,10.1994],[70,10.437],[71,11.1311],[72,11.6958],[73,11.6279],[74,11.4784],[76,10.8778],[77,9.9737],[78,9.4044],[83,9.3156],[84,9.0265],[85,8.791],[86,8.4315],[87,8.4453],[88,8.0656],[89,8.004],[90,7.9793],[91,7.5146],[92,7.8627],[93,7.7191],[94,8.0253],[95,7.9329],[96,9.0659],[97,11.2922],[98,13.9633],[100,19.6513],[101,9.8142],[102,-3.9902],[103,16.7481],[104,18.5545],[105,17.665],[106,13.7323],[107,13.6411],[108,13.2742],[109,13.4681],[110,13.9156],[111,14.625],[112,25.7217],[113,29.0125],[114,30.6758],[115,32.4176],[117,34.6041],[118,35.4754],[119,36.4539],[121,37.2966],[122,37.8879],[123,38.1997],[124,38.8002],[125,38.5627],[127,38.6427],[128,38.4713],[129,38.4266],[131,38.6599],[133,38.7401],[134,38.5353],[135,38.6094],[137,39.4215],[138,39.9508],[139,40.1185],[140,40.204],[142,40.1429],[143,40.2032],[144,40.0734],[145,40.0923],[146,39.992],[147,40.0073],[148,39.9084],[149,39.9293],[150,40.0347],[151,39.8137],[153,39.7295],[154,39.7605],[155,39.6415],[156,39.7699],[157,39.5996],[158,39.6167],[160,39.3193],[164,39.5813],[165,39.5629],[166,39.244],[167,35.6193],[168,27.0926],[169,20.2807],[170,13.9043],[171,10.1191],[172,9.1755],[173,8.3691],[174,8.419],[175,8.533],[176,8.5055],[177,8.1032],[178,7.5332],[179,7.6146],[180,7.0354],[181,6.8397],[182,6.7503],[183,6.57],[184,5.3056]]
       }
       index_03_r_ctrl.Scale.X {type=MovieSceneScriptingFloatChannel, default=1, const [1,185]}
       index_03_r_ctrl.Scale.Y {type=MovieSceneScriptingFloatChannel, default=1, const [1,185]}
       index_03_r_ctrl.Scale.Z {type=MovieSceneScriptingFloatChannel, default=1, const [1,185]}
       middle_metacarpal_r_ctrl.Location.X {
        type=MovieSceneScriptingFloatChannel
        default=-0.0001
        thinned_from=185
        keys [[0,-0.0001],[51,-0.0002],[56,-0.0004],[61,-0.0012],[68,-0.0005],[70,-0.0002],[72,-0.0004],[99,-0.0001],[100,-0.0004],[101,-0.0004],[102,-0.0012],[103,-0.0005],[104,-0.0001],[105,-0.0001],[106,-0.0007],[107,-0.0001],[108,-0.0046],[109,-0.0007],[110,-0.0033],[111,-0.0006],[112,-0.0013],[113,-0.0012],[114,-0.0016],[115,-0.0007],[118,-0.0002],[135,-0.0002],[136,-0.0011],[137,-0.0001],[138,-0.0006],[139,-0.0001],[140,-0.0003],[168,-0.0001],[169,-0.0002],[170,-0.0008],[171,-0.0009],[172,-0.0002],[180,-0.0002],[181,-0.0004],[182,-0.0003],[183,-0.0007],[184,-0.0001]]
       }
       middle_metacarpal_r_ctrl.Location.Y {
        type=MovieSceneScriptingFloatChannel
        default=0.0002
        thinned_from=185
        keys [[0,0.0002],[56,-0.0001],[63,-0.0006],[105,0.0002],[106,0.0005],[107,0.0002],[108,0.002],[109,0.0005],[110,0.0023],[111,0.0003],[112,0.0008],[113,0],[135,0.0002],[136,0.0004],[137,0.0002],[138,0.0005],[139,0.0002],[169,0.0002],[181,0.0002],[183,0.0006],[184,0.0002]]
       }
       middle_metacarpal_r_ctrl.Location.Z {
        type=MovieSceneScriptingFloatChannel
        default=0.0001
        thinned_from=185
        keys [[0,0.0001],[48,0.0001],[56,0.0004],[61,0.0012],[70,0.0001],[101,0.0002],[102,0.0009],[104,0.0001],[107,0.0001],[108,0.0002],[109,0],[110,0.0001],[111,0.0007],[112,0.0016],[113,0.0001],[114,0.0001],[135,0.0001],[136,0.0004],[137,0.0001],[138,0.0003],[139,0.0001],[169,0.0001],[171,0.0004],[172,0.0001],[184,0.0001]]
       }
       middle_metacarpal_r_ctrl.Rotation.X {type=MovieSceneScriptingFloatChannel, default=0.07, const [0.07,185]}
       middle_metacarpal_r_ctrl.Rotation.Y {type=MovieSceneScriptingFloatChannel, default=1.7355, const [1.7355,185]}
       middle_metacarpal_r_ctrl.Rotation.Z {type=MovieSceneScriptingFloatChannel, default=-2.244, const [-2.244,185]}
       middle_metacarpal_r_ctrl.Scale.X {type=MovieSceneScriptingFloatChannel, default=1, const [1,185]}
       middle_metacarpal_r_ctrl.Scale.Y {type=MovieSceneScriptingFloatChannel, default=1, const [1,185]}
       middle_metacarpal_r_ctrl.Scale.Z {type=MovieSceneScriptingFloatChannel, default=1, const [1,185]}
       middle_01_r_ctrl.Location.X {
        type=MovieSceneScriptingFloatChannel
        default=0.0001
        thinned_from=185
        keys [[0,0.0002],[55,-0.0006],[57,-0.0012],[60,-0.0024],[62,-0.0028],[64,-0.0025],[68,-0.001],[70,0],[72,-0.0007],[74,0.0002],[82,-0.0001],[83,0.0002],[99,0],[100,-0.0007],[101,-0.0005],[102,-0.0027],[103,-0.001],[104,0],[105,0.0001],[106,-0.0022],[107,0.0002],[108,-0.0159],[109,-0.0021],[110,-0.013],[111,-0.001],[112,-0.0027],[113,-0.0036],[114,-0.0049],[115,-0.0017],[116,-0.0012],[117,-0.0004],[118,-0.0001],[133,0.0002],[134,0],[135,0.0001],[136,-0.0033],[137,0.0002],[138,-0.0018],[139,0.0001],[140,-0.0004],[141,0.0002],[165,0.0002],[166,0],[167,0.0002],[169,-0.0005],[170,-0.0019],[171,-0.0024],[172,0],[180,0.0001],[181,-0.0007],[182,-0.0009],[183,-0.0022],[184,0.0001]]
       }
       middle_01_r_ctrl.Location.Y {
        type=MovieSceneScriptingFloatChannel
        default=0
        thinned_from=185
        keys [[0,0],[49,-0.0003],[55,-0.0011],[56,-0.0014],[60,-0.0036],[61,-0.004],[62,-0.0042],[63,-0.004],[65,-0.003],[69,-0.0006],[70,-0.0002],[72,-0.0003],[74,0],[97,0],[99,0.0002],[100,0],[101,0],[102,-0.0005],[103,-0.0001],[104,0.0002],[105,0],[106,0.0004],[108,-0.0005],[109,0.0001],[110,0.0019],[111,0.0002],[112,0.0011],[113,-0.0008],[114,-0.0007],[115,-0.0001],[133,0],[134,0.0002],[135,0],[137,0],[138,0.0007],[139,0.0001],[140,0.0003],[141,0],[168,-0.0001],[170,0.0003],[181,-0.0001],[182,0.0005],[183,0.0006],[184,0]]
       }
       middle_01_r_ctrl.Location.Z {
        type=MovieSceneScriptingFloatChannel
        default=0
        thinned_from=185
        keys [[0,0],[55,0.0005],[61,0.0016],[64,0.0015],[70,0.0001],[72,0.0004],[74,0],[77,0.0001],[82,-0.0002],[99,0.0002],[100,0.0003],[101,0.0002],[102,0.0024],[103,0.001],[104,-0.0001],[105,0],[106,-0.0005],[107,0],[108,-0.0028],[109,-0.0006],[110,-0.0025],[111,0.0015],[112,0.0036],[113,0],[114,-0.0004],[115,0.0002],[136,-0.0001],[137,0],[138,-0.0001],[140,0.0001],[168,0],[169,0.0001],[171,0.0006],[172,0.0001],[180,0.0001],[181,0.0004],[182,-0.0004],[183,-0.0005],[184,0]]
       }
       middle_01_r_ctrl.Rotation.X {
        type=MovieSceneScriptingFloatChannel
        default=-9.4047
        thinned_from=185
        keys [[0,-9.5066],[2,-9.4931],[3,-9.5008],[4,-9.4835],[5,-9.5056],[14,-9.5061],[25,-9.5467],[28,-9.538],[29,-9.5214],[31,-9.5392],[37,-9.527],[41,-9.5467],[42,-9.5287],[44,-9.5525],[47,-9.5147],[55,-9.1794],[56,-9.0821],[57,-8.9125],[59,-8.4797],[60,-8.2783],[61,-8.1279],[62,-8.0594],[67,-7.9933],[70,-7.9976],[77,-7.8712],[78,-7.8364],[82,-7.8406],[83,-7.8951],[85,-7.9101],[86,-7.8941],[88,-7.928],[91,-7.9041],[92,-7.8811],[94,-7.9134],[97,-7.8493],[99,-7.728],[100,-7.8141],[101,-7.2175],[102,-7.1641],[103,-7.2402],[104,-7.8403],[105,-7.8691],[106,-7.4972],[107,-7.5043],[108,-7.4282],[109,-7.4419],[110,-7.419],[111,-7.4369],[112,-7.931],[113,-8.1019],[114,-8.1394],[115,-8.087],[117,-8.0279],[118,-7.9714],[119,-7.9396],[123,-7.8704],[124,-7.8889],[127,-7.873],[129,-7.914],[132,-7.892],[134,-7.8461],[136,-7.8712],[137,-7.8649],[139,-7.9758],[140,-7.9812],[141,-8.0098],[152,-8.0298],[154,-8.0686],[162,-8.0547],[163,-8.0424],[164,-8.0523],[166,-8.038],[167,-8.067],[168,-8.2021],[169,-8.767],[170,-9.2091],[171,-9.5921],[172,-9.7119],[173,-9.8056],[175,-9.8216],[176,-9.7983],[177,-9.8147],[180,-9.7875],[181,-9.728],[182,-9.1689],[183,-9.249],[184,-9.4047]]
       }
       middle_01_r_ctrl.Rotation.Y {
        type=MovieSceneScriptingFloatChannel
        default=-6.421
        thinned_from=185
        keys [[0,-7.6781],[1,-7.6703],[2,-7.6388],[4,-7.681],[5,-7.6274],[6,-7.6784],[7,-7.6925],[8,-7.583],[9,-7.5795],[10,-7.5184],[11,-7.5217],[12,-7.4726],[13,-7.5672],[14,-7.5577],[15,-7.5668],[16,-7.6812],[17,-7.7159],[19,-7.7171],[20,-7.6749],[21,-7.6938],[22,-7.7379],[24,-7.6662],[25,-7.7659],[26,-7.7002],[27,-7.703],[28,-7.6711],[29,-7.6763],[30,-7.6585],[31,-7.696],[32,-7.6441],[33,-7.6718],[34,-7.622],[35,-7.718],[36,-7.6807],[37,-7.7007],[38,-7.6796],[39,-7.7121],[40,-7.7282],[41,-7.6686],[42,-7.77],[43,-7.716],[44,-7.7625],[45,-7.748],[46,-7.7063],[47,-7.6401],[48,-7.5521],[49,-7.445],[51,-7.1846],[54,-6.7192],[55,-6.555],[56,-6.2408],[57,-5.6942],[58,-5.0147],[59,-4.3021],[60,-3.6561],[61,-3.1764],[62,-2.9627],[64,-2.8714],[65,-2.8323],[66,-2.8011],[67,-2.7805],[68,-2.773],[69,-2.8048],[70,-2.8747],[71,-3.1013],[72,-3.2965],[73,-3.3086],[74,-3.3141],[75,-3.094],[76,-2.9537],[77,-2.8002],[78,-2.7282],[79,-2.7452],[81,-2.82],[82,-2.837],[83,-2.7778],[84,-2.7725],[85,-2.7162],[86,-2.5749],[87,-2.7288],[88,-2.7884],[89,-2.7515],[90,-2.7675],[91,-2.5037],[92,-2.2658],[93,-2.3007],[94,-2.2327],[95,-2.2125],[96,-2.3817],[97,-2.6928],[98,-2.7161],[99,-2.1385],[100,-1.3364],[101,-1.4715],[102,1.3708],[103,-2.618],[104,-2.5668],[105,-2.4093],[106,0.7302],[107,0.9122],[108,0.9483],[109,0.888],[110,0.3781],[111,-1.2169],[112,-4.1283],[113,-5.85],[114,-7.647],[115,-8.8375],[116,-9.2795],[117,-9.1552],[118,-9.0735],[119,-8.953],[120,-8.883],[121,-8.8584],[123,-8.6806],[124,-8.6087],[125,-8.6476],[127,-8.648],[128,-8.6855],[131,-8.6689],[132,-8.6754],[133,-8.7708],[134,-8.7516],[135,-8.7045],[136,-8.4988],[137,-8.0409],[138,-7.7214],[139,-7.7484],[140,-7.6883],[141,-7.696],[142,-7.8167],[143,-7.8557],[144,-7.8505],[145,-7.7761],[147,-7.8337],[148,-7.8249],[149,-7.8722],[150,-7.858],[151,-7.9006],[153,-8.0421],[154,-8.1272],[156,-8.1283],[157,-8.1522],[158,-8.11],[159,-8.1096],[160,-8.1492],[161,-8.1634],[162,-8.1596],[163,-8.1325],[164,-7.7937],[165,-6.9137],[166,-5.7877],[167,-4.9066],[168,-3.0851],[169,-2.7053],[170,-3.8717],[171,-5.7967],[172,-6.2916],[173,-6.6097],[174,-6.7116],[175,-6.6904],[176,-6.7583],[178,-6.6583],[179,-6.7039],[180,-6.7839],[181,-6.7569],[182,-5.2637],[183,-5.6349],[184,-6.421]]
       }
       middle_01_r_ctrl.Rotation.Z {
        type=MovieSceneScriptingFloatChannel
        default=-74.5664
        thinned_from=185
        keys [[0,-77.8727],[2,-77.6285],[3,-77.7664],[4,-77.5614],[5,-77.7819],[6,-77.8354],[9,-77.4853],[11,-77.5604],[12,-77.4382],[16,-78.104],[19,-78.2647],[20,-78.1586],[22,-78.3689],[24,-78.3705],[25,-78.5537],[26,-78.3159],[28,-78.2922],[29,-78.0724],[30,-78.0816],[31,-78.3454],[32,-78.1714],[33,-78.1697],[34,-77.986],[36,-78.2218],[37,-78.1855],[38,-78.3397],[41,-78.407],[42,-78.3143],[44,-78.6288],[45,-78.5424],[47,-77.9003],[49,-76.74],[52,-74.3116],[55,-71.4461],[56,-69.5704],[57,-66.2983],[58,-62.2241],[59,-57.9419],[60,-54.046],[61,-51.1306],[62,-49.7898],[65,-48.8126],[67,-48.4244],[68,-48.3685],[70,-48.6589],[72,-48.746],[73,-48.6402],[74,-48.4076],[77,-46.5466],[78,-45.8552],[82,-46.1199],[83,-46.8786],[84,-47.0834],[85,-47.0058],[86,-46.4975],[88,-47.4176],[89,-47.2371],[90,-47.153],[91,-46.5254],[92,-45.7333],[94,-46.1898],[95,-45.8515],[97,-45.9987],[98,-45.1432],[99,-42.9661],[100,-42.8272],[101,-32.5785],[102,-25.2858],[103,-35.5775],[104,-45.6208],[105,-45.8007],[106,-33.3225],[107,-33.0689],[108,-31.4844],[109,-31.9348],[110,-31.3724],[111,-32.1054],[112,-42.858],[113,-44.1397],[114,-44.2403],[115,-42.7685],[116,-42.1289],[117,-41.2736],[118,-40.1741],[119,-39.4006],[120,-39.0277],[121,-38.7482],[123,-37.6461],[124,-37.8165],[125,-37.8056],[126,-37.6129],[127,-37.5801],[128,-38.0539],[129,-38.3374],[133,-37.7605],[134,-37.2285],[135,-37.3205],[136,-37.2122],[137,-36.1486],[138,-36.4497],[139,-37.5129],[140,-37.4828],[141,-37.9835],[142,-38.2671],[144,-38.3112],[145,-38.1804],[147,-38.4238],[148,-38.438],[149,-38.6522],[150,-38.6227],[152,-38.7643],[153,-39.1177],[154,-39.7148],[155,-39.5768],[156,-39.6112],[158,-39.4549],[159,-39.4527],[160,-39.5302],[162,-39.5196],[163,-39.2602],[164,-39.4883],[165,-39.552],[166,-40.0248],[167,-42.0687],[168,-46.7222],[169,-58.6668],[170,-67.9069],[171,-76.1118],[172,-78.581],[173,-80.3504],[174,-80.7351],[175,-80.6864],[176,-80.4721],[177,-80.6422],[178,-80.3716],[180,-80.3672],[181,-79.4945],[182,-69.5754],[183,-71.1193],[184,-74.5664]]
       }
       middle_01_r_ctrl.Scale.X {type=MovieSceneScriptingFloatChannel, default=1, const [1,185]}
       middle_01_r_ctrl.Scale.Y {type=MovieSceneScriptingFloatChannel, default=1, const [1,185]}
       middle_01_r_ctrl.Scale.Z {type=MovieSceneScriptingFloatChannel, default=1, const [1,185]}
       middle_02_r_ctrl.Location.X {
        type=MovieSceneScriptingFloatChannel
        default=0.0008
        thinned_from=185
        keys [[0,0.0008],[56,0.0004],[62,-0.0004],[70,0.0007],[72,0.0005],[74,0.0008],[82,0.0006],[97,0.0007],[99,0.0004],[100,-0.0023],[101,-0.0018],[102,-0.0011],[103,0],[104,0.0006],[105,0.0007],[106,-0.0001],[107,0.0008],[108,-0.0034],[109,-0.0001],[110,-0.0048],[111,-0.0006],[112,-0.0026],[113,0.0002],[114,-0.0001],[115,0.0004],[133,0.0008],[134,0.0006],[135,0.0007],[136,0.0003],[137,0.0008],[138,0.0001],[139,0.0007],[140,0.0005],[141,0.0008],[165,0.0008],[166,0.0005],[167,0.0008],[170,0],[171,0.0001],[172,0.0007],[180,0.0007],[181,0.0002],[182,0.0003],[183,0.0008],[184,0.0008]]
       }
       middle_02_r_ctrl.Location.Y {
        type=MovieSceneScriptingFloatChannel
        default=0.0008
        thinned_from=185
        keys [[0,0.0009],[63,0.0015],[72,0.0007],[82,0.0007],[97,0.0008],[99,0.0006],[100,-0.0024],[101,-0.0013],[102,0.0005],[105,0.0008],[106,-0.0004],[107,0.0009],[108,-0.006],[109,-0.0002],[110,-0.0047],[111,0.0005],[112,-0.0002],[114,-0.0009],[115,0],[118,0.0007],[133,0.0009],[134,0.0007],[135,0.0008],[136,-0.0009],[137,0.0009],[138,-0.0002],[139,0.0008],[140,0.0005],[141,0.0008],[165,0.0009],[166,0.0006],[167,0.0009],[169,0.0001],[170,-0.0006],[171,-0.0007],[172,0.0007],[180,0.0008],[181,0.0002],[182,0],[183,0.0003],[184,0.0008]]
       }
       middle_02_r_ctrl.Location.Z {
        type=MovieSceneScriptingFloatChannel
        default=0.0476
        thinned_from=185
        keys [[0,0.0476],[68,0.048],[70,0.0477],[72,0.0479],[99,0.0476],[100,0.0478],[102,0.0473],[104,0.0476],[105,0.0476],[106,0.0479],[107,0.0476],[108,0.0496],[109,0.0478],[110,0.0493],[111,0.0473],[112,0.0466],[113,0.0469],[114,0.0466],[115,0.0473],[135,0.0476],[136,0.0479],[137,0.0476],[166,0.0475],[167,0.0476],[169,0.0472],[170,0.0468],[171,0.0467],[172,0.0475],[181,0.0476],[182,0.0479],[184,0.0476]]
       }
       middle_02_r_ctrl.Rotation.X {
        type=MovieSceneScriptingFloatChannel
        default=-3.166
        thinned_from=185
        keys [[0,-5.7762],[1,-5.8354],[2,-5.8553],[3,-5.7875],[4,-5.8529],[5,-5.8018],[6,-5.7768],[7,-5.8015],[8,-5.8412],[9,-5.7736],[10,-5.82],[11,-5.8347],[12,-5.8995],[14,-5.8421],[15,-5.8396],[16,-5.751],[18,-5.6965],[20,-5.7163],[21,-5.6673],[23,-5.6471],[24,-5.6456],[25,-5.5428],[26,-5.5843],[27,-5.5861],[28,-5.5498],[29,-5.5959],[30,-5.5882],[31,-5.5219],[32,-5.5671],[33,-5.551],[34,-5.5982],[35,-5.5627],[36,-5.5963],[38,-5.5688],[39,-5.5676],[40,-5.5509],[41,-5.5829],[42,-5.5888],[44,-5.519],[50,-5.1774],[52,-5.0433],[54,-4.8816],[55,-4.787],[56,-4.6068],[57,-4.3023],[58,-3.9286],[59,-3.5404],[60,-3.1926],[61,-2.9401],[62,-2.8378],[68,-2.8021],[70,-2.6993],[71,-2.8264],[72,-2.9828],[74,-3.0609],[76,-3.3242],[77,-3.6819],[78,-3.8994],[80,-3.9486],[82,-3.9787],[83,-3.9858],[84,-4.183],[85,-4.2688],[86,-4.3049],[87,-4.3505],[89,-4.4904],[90,-4.4891],[91,-4.5227],[92,-4.5171],[93,-4.5418],[94,-4.3595],[95,-4.5352],[96,-4.3697],[97,-3.9568],[98,-3.2653],[99,-2.4796],[100,-0.7158],[101,-2.5623],[102,-6.0145],[103,-2.5996],[104,0.0728],[105,0.2789],[106,-4.8668],[107,-4.8926],[108,-5.1546],[109,-5.153],[110,-5.2807],[111,-5.2678],[112,0.7979],[113,1.3125],[114,1.4023],[115,1.4431],[116,1.4519],[117,1.4458],[118,1.4319],[119,1.4092],[120,1.3837],[121,1.3707],[123,1.3202],[124,1.3069],[125,1.3202],[126,1.3087],[127,1.3037],[128,1.3109],[129,1.3131],[130,1.3127],[132,1.3033],[133,1.3039],[134,1.2899],[135,1.2805],[137,1.241],[138,1.2109],[139,1.2075],[140,1.1936],[141,1.1874],[142,1.19],[143,1.1867],[144,1.1894],[145,1.2042],[147,1.2085],[150,1.2084],[151,1.2149],[152,1.2152],[154,1.211],[155,1.2212],[156,1.2175],[158,1.2229],[160,1.2377],[161,1.2346],[162,1.228],[163,1.2297],[164,1.2213],[165,1.2276],[166,1.2582],[167,1.4274],[168,1.2061],[169,0.3757],[170,-0.9663],[171,-1.5428],[172,-1.6955],[173,-1.769],[174,-1.8118],[175,-1.8086],[176,-1.8548],[177,-1.8773],[178,-2.0708],[179,-2.1597],[180,-2.3223],[181,-2.4604],[182,-2.8052],[183,-2.9246],[184,-3.166]]
       }
       middle_02_r_ctrl.Rotation.Y {
        type=MovieSceneScriptingFloatChannel
        default=8.9189
        thinned_from=185
        keys [[0,10.3837],[2,10.415],[3,10.388],[4,10.4141],[6,10.3839],[8,10.4096],[9,10.3827],[12,10.4323],[15,10.4089],[16,10.3735],[18,10.3514],[20,10.3594],[24,10.3304],[25,10.2871],[27,10.3055],[28,10.2901],[29,10.3096],[30,10.3064],[31,10.2781],[32,10.2974],[33,10.2906],[34,10.3106],[35,10.2956],[40,10.2905],[42,10.3066],[49,10.1251],[54,9.9015],[55,9.8449],[56,9.737],[57,9.5534],[59,9.0919],[60,8.8806],[61,8.7267],[62,8.6633],[68,8.6355],[70,8.5503],[71,8.653],[72,8.7783],[74,8.8389],[76,9.0343],[77,9.2806],[78,9.4238],[83,9.478],[84,9.5978],[85,9.6479],[89,9.7736],[93,9.8019],[94,9.7002],[95,9.7984],[96,9.7055],[97,9.4587],[98,8.9885],[99,8.3579],[100,6.4285],[101,8.4405],[102,10.4128],[103,8.4324],[104,5.1575],[105,4.7499],[106,9.9722],[107,9.9855],[108,10.113],[109,10.1116],[110,10.1679],[111,10.083],[112,2.7985],[113,1.4386],[114,0.7053],[115,0.0831],[116,-0.3072],[117,-0.7587],[119,-1.3696],[120,-1.635],[121,-1.7492],[122,-1.9434],[123,-2.1219],[124,-2.2073],[125,-2.1223],[126,-2.1964],[127,-2.2276],[128,-2.1823],[129,-2.168],[130,-2.171],[131,-2.2083],[132,-2.2301],[133,-2.2262],[134,-2.3126],[135,-2.3661],[136,-2.4843],[137,-2.5814],[138,-2.7331],[139,-2.7488],[140,-2.8165],[141,-2.8428],[142,-2.8308],[143,-2.8461],[144,-2.8337],[145,-2.7639],[147,-2.7435],[149,-2.7484],[150,-2.7441],[151,-2.7128],[152,-2.7112],[154,-2.7314],[155,-2.6817],[156,-2.6997],[157,-2.6832],[158,-2.6731],[160,-2.5986],[161,-2.6144],[162,-2.6474],[163,-2.6391],[164,-2.6813],[165,-2.6495],[166,-2.4998],[167,-1.1396],[168,1.9609],[169,4.5202],[170,6.7458],[171,7.4407],[172,7.6071],[173,7.6844],[174,7.7287],[175,7.7253],[176,7.7728],[177,7.7955],[178,7.9868],[179,8.0717],[180,8.2225],[181,8.3462],[182,8.6367],[183,8.734],[184,8.9189]]
       }
       middle_02_r_ctrl.Rotation.Z {
        type=MovieSceneScriptingFloatChannel
        default=-23.3782
        thinned_from=185
        keys [[0,-38.3127],[1,-38.63],[2,-38.7362],[3,-38.3727],[4,-38.7233],[5,-38.4497],[6,-38.3155],[7,-38.4479],[8,-38.6608],[9,-38.2985],[10,-38.5475],[11,-38.6262],[12,-38.9725],[14,-38.666],[15,-38.6522],[16,-38.1768],[18,-37.8842],[20,-37.9904],[21,-37.7268],[24,-37.61],[25,-37.0541],[26,-37.2788],[27,-37.2887],[28,-37.0922],[29,-37.3417],[30,-37.2998],[31,-36.9413],[32,-37.1856],[33,-37.0985],[34,-37.3538],[35,-37.1621],[36,-37.3437],[38,-37.1952],[39,-37.1884],[40,-37.0979],[41,-37.2712],[42,-37.3032],[44,-36.9254],[50,-34.972],[52,-34.1799],[54,-33.2299],[55,-32.6793],[56,-31.6306],[57,-29.8559],[58,-27.6757],[59,-25.4104],[60,-23.3803],[61,-21.9058],[62,-21.3072],[68,-21.0923],[70,-20.4227],[71,-21.2313],[72,-22.2251],[74,-22.719],[76,-24.3486],[77,-26.5081],[78,-27.8036],[80,-28.0932],[82,-28.2699],[83,-28.3112],[84,-29.459],[85,-29.953],[87,-30.4227],[89,-31.2213],[90,-31.2137],[91,-31.4046],[92,-31.3726],[93,-31.5135],[94,-30.4745],[95,-31.4761],[96,-30.5306],[97,-28.1361],[98,-23.9788],[99,-18.973],[100,-5.9875],[101,-19.5421],[102,-39.3344],[103,-19.6301],[104,1.3773],[105,3.6207],[106,-33.3409],[107,-33.4843],[108,-34.9351],[109,-34.9266],[110,-35.6276],[111,-35.1629],[112,13.6816],[113,20.6516],[114,24.252],[115,27.3235],[116,29.2331],[117,31.4392],[119,34.4338],[120,35.7397],[121,36.3001],[122,37.2589],[123,38.1405],[124,38.563],[125,38.1419],[126,38.509],[127,38.6638],[128,38.4394],[129,38.3684],[130,38.3833],[132,38.6761],[133,38.6568],[134,39.085],[135,39.3501],[136,39.9431],[137,40.4206],[138,41.1777],[139,41.2561],[140,41.595],[141,41.7261],[142,41.6662],[143,41.7429],[144,41.6804],[145,41.3315],[147,41.2294],[150,41.2325],[151,41.0759],[154,41.1689],[155,40.921],[156,41.0105],[158,40.878],[160,40.5064],[161,40.5852],[162,40.7497],[163,40.7082],[164,40.9186],[165,40.7605],[166,39.9964],[167,33.3111],[168,18.0746],[169,4.7963],[170,-8.0434],[171,-12.4852],[172,-13.5868],[173,-14.1153],[174,-14.4194],[175,-14.3967],[176,-14.7245],[177,-14.8829],[178,-16.2314],[179,-16.8413],[180,-17.9421],[181,-18.8604],[182,-21.1015],[183,-21.8531],[184,-23.3782]]
       }
       middle_02_r_ctrl.Scale.X {type=MovieSceneScriptingFloatChannel, default=1, const [1,185]}
       middle_02_r_ctrl.Scale.Y {type=MovieSceneScriptingFloatChannel, default=1, const [1,185]}
       middle_02_r_ctrl.Scale.Z {type=MovieSceneScriptingFloatChannel, default=1, const [1,185]}
       middle_03_r_ctrl.Location.X {
        type=MovieSceneScriptingFloatChannel
        default=0.0005
        thinned_from=185
        keys [[0,0.0005],[55,-0.0002],[61,-0.0022],[62,-0.0023],[64,-0.0019],[69,0.0002],[70,0.0004],[72,0.0001],[74,0.0005],[82,0.0004],[96,0.0004],[97,0.0003],[98,-0.0002],[99,-0.0005],[100,0.0005],[101,0.0005],[102,-0.0045],[103,-0.0017],[104,0.0003],[105,0.0005],[106,0.0003],[107,0.0005],[108,-0.0017],[109,0.0002],[110,-0.0016],[111,-0.011],[112,-0.0324],[113,-0.0008],[114,-0.0013],[115,-0.0001],[118,0.0004],[133,0.0005],[134,0.0004],[135,0.0005],[136,-0.0002],[137,0.0005],[138,-0.0001],[139,0.0005],[140,0.0003],[141,0.0005],[165,0.0005],[166,-0.0016],[167,0.0005],[168,-0.0009],[169,-0.0012],[170,-0.002],[171,-0.0011],[172,0.0004],[180,0.0005],[181,0.0002],[183,0.0002],[184,0.0005]]
       }
       middle_03_r_ctrl.Location.Y {
        type=MovieSceneScriptingFloatChannel
        default=0
        thinned_from=185
        keys [[0,0],[64,-0.0004],[97,0],[98,-0.0003],[99,-0.0003],[100,0],[101,0],[102,-0.0033],[103,-0.0009],[104,-0.0001],[108,0.0002],[109,-0.0001],[110,-0.0001],[111,-0.0075],[112,-0.0066],[113,-0.0008],[114,-0.0008],[115,-0.0005],[118,0],[135,0],[136,-0.0009],[137,0],[138,-0.0004],[139,0],[140,-0.0001],[165,0],[166,-0.0007],[167,0],[170,-0.001],[172,0],[180,0],[181,-0.0001],[183,0.0003],[184,0]]
       }
       middle_03_r_ctrl.Location.Z {
        type=MovieSceneScriptingFloatChannel
        default=0.0061
        thinned_from=185
        keys [[0,0.0061],[97,0.0061],[99,0.0059],[101,0.0061],[102,0.0049],[103,0.0054],[104,0.0062],[105,0.0061],[106,0.0065],[107,0.0061],[108,0.0086],[109,0.0065],[110,0.0084],[111,0.0064],[112,0.0011],[113,0.0053],[114,0.0049],[115,0.0057],[135,0.0061],[136,0.0063],[165,0.0061],[166,0.0059],[167,0.0061],[169,0.0056],[170,0.0051],[171,0.0052],[172,0.006],[181,0.006],[182,0.0064],[184,0.0061]]
       }
       middle_03_r_ctrl.Rotation.X {
        type=MovieSceneScriptingFloatChannel
        default=-3.1439
        thinned_from=185
        keys [[0,-3.9107],[3,-3.9229],[4,-3.943],[5,-3.9169],[6,-3.9098],[8,-3.9171],[9,-3.9115],[10,-3.9217],[11,-3.9121],[12,-3.9347],[13,-3.9276],[15,-3.9348],[17,-3.8951],[19,-3.9045],[21,-3.8971],[22,-3.9018],[23,-3.8703],[24,-3.8641],[25,-3.8402],[26,-3.8492],[27,-3.8472],[28,-3.837],[29,-3.8557],[31,-3.8354],[32,-3.8455],[33,-3.8374],[34,-3.8586],[35,-3.8469],[38,-3.8491],[39,-3.856],[40,-3.8447],[42,-3.8534],[45,-3.8276],[48,-3.7739],[51,-3.691],[55,-3.5468],[56,-3.4718],[57,-3.3417],[58,-3.18],[59,-3.0099],[60,-2.855],[61,-2.7386],[62,-2.6841],[68,-2.6209],[69,-2.6014],[70,-2.57],[71,-2.4923],[72,-2.4337],[73,-2.461],[75,-2.5634],[76,-2.5989],[77,-2.6905],[78,-2.7386],[82,-2.7551],[83,-2.7342],[84,-2.7907],[85,-2.8277],[87,-2.8768],[88,-2.9245],[89,-2.9534],[90,-2.9535],[91,-2.8756],[92,-2.738],[93,-2.7635],[94,-2.7151],[96,-2.8386],[97,-2.7671],[98,-2.6243],[99,-2.4219],[100,-2.0247],[101,-3.4632],[102,-2.4261],[103,-3.001],[104,-2.2548],[105,-2.2207],[106,-2.9154],[107,-2.8238],[108,-2.8831],[109,-2.8707],[110,-2.9149],[111,-3.0547],[112,-1.593],[113,-1.3898],[114,-1.328],[115,-1.2771],[116,-1.2513],[117,-1.2285],[118,-1.2123],[119,-1.2006],[120,-1.1915],[123,-1.1781],[124,-1.1754],[127,-1.1741],[128,-1.1771],[133,-1.1747],[136,-1.1689],[166,-1.1712],[167,-1.2076],[168,-1.4084],[169,-1.7696],[170,-2.2101],[171,-2.4455],[172,-2.5116],[173,-2.5845],[174,-2.5839],[175,-2.5725],[176,-2.6217],[177,-2.6124],[178,-2.6477],[179,-2.6926],[180,-2.731],[181,-2.7931],[182,-2.9207],[183,-2.9871],[184,-3.1439]]
       }
       middle_03_r_ctrl.Rotation.Y {
        type=MovieSceneScriptingFloatChannel
        default=5.5567
        thinned_from=185
        keys [[0,6.394],[3,6.4058],[4,6.4249],[5,6.4],[9,6.3949],[10,6.4046],[11,6.3954],[12,6.4171],[13,6.4103],[15,6.4172],[17,6.379],[22,6.3855],[23,6.355],[24,6.3489],[25,6.3256],[27,6.3324],[28,6.3223],[29,6.3407],[31,6.3208],[32,6.3307],[33,6.3228],[34,6.3435],[38,6.3342],[39,6.341],[40,6.33],[42,6.3385],[46,6.2964],[50,6.1872],[55,5.9782],[56,5.8885],[57,5.7324],[59,5.3324],[60,5.1444],[61,5.0016],[62,4.9321],[68,4.8375],[69,4.8065],[70,4.756],[71,4.6286],[72,4.5319],[73,4.5781],[75,4.7448],[76,4.8005],[77,4.9393],[78,5.0108],[82,5.0349],[83,5.0044],[84,5.0861],[85,5.1384],[87,5.2067],[88,5.2718],[89,5.3105],[90,5.3108],[91,5.2051],[92,5.0098],[93,5.047],[94,4.9764],[96,5.1535],[97,5.0517],[98,4.8388],[99,4.513],[100,3.7685],[101,5.9344],[102,4.5487],[103,5.3836],[104,4.2204],[105,4.1565],[106,5.2583],[107,5.1328],[108,5.2148],[109,5.1983],[110,5.2596],[111,5.3879],[112,2.4089],[113,1.9163],[115,1.3222],[116,1.1464],[117,0.9666],[118,0.8171],[119,0.6873],[120,0.5738],[121,0.5104],[122,0.4249],[123,0.3445],[124,0.283],[125,0.2977],[126,0.2657],[127,0.2486],[128,0.3248],[129,0.3093],[130,0.2799],[131,0.2785],[132,0.2873],[133,0.2638],[134,0.213],[135,0.1811],[136,0.1249],[137,0.0263],[138,-0.0626],[139,-0.0331],[140,-0.0516],[141,-0.0418],[142,-0.0234],[143,-0.0211],[144,-0.0185],[145,-0.0133],[146,-0.006],[147,0.005],[148,0.0053],[149,0.0233],[150,0.0163],[151,0.0342],[152,0.0421],[153,0.0552],[154,0.0673],[155,0.0718],[156,0.0747],[157,0.0975],[158,0.0904],[159,0.1161],[160,0.1375],[161,0.1295],[162,0.1274],[163,0.1047],[164,0.1048],[165,0.1031],[166,0.1884],[167,0.7692],[168,1.988],[169,3.163],[170,4.1451],[171,4.5577],[172,4.6627],[173,4.7786],[174,4.7772],[175,4.7594],[176,4.8359],[177,4.8215],[178,4.8755],[179,4.9429],[180,4.9999],[181,5.0894],[182,5.2667],[183,5.3546],[184,5.5567]]
       }
       middle_03_r_ctrl.Rotation.Z {
        type=MovieSceneScriptingFloatChannel
        default=-29.178
        thinned_from=185
        keys [[0,-36.4461],[3,-36.5551],[4,-36.7323],[5,-36.5009],[6,-36.4385],[8,-36.5032],[9,-36.4538],[10,-36.544],[11,-36.4588],[12,-36.6595],[13,-36.5966],[15,-36.6603],[17,-36.3077],[19,-36.3911],[21,-36.3259],[22,-36.367],[23,-36.0866],[24,-36.0306],[25,-35.8172],[26,-35.8978],[27,-35.8793],[28,-35.7879],[29,-35.9556],[31,-35.774],[32,-35.8643],[33,-35.7919],[34,-35.9816],[35,-35.8766],[38,-35.8966],[39,-35.9586],[40,-35.8575],[42,-35.9354],[45,-35.7004],[47,-35.3857],[50,-34.6487],[54,-33.2915],[55,-32.9044],[56,-32.1558],[57,-30.8543],[58,-29.2333],[59,-27.526],[60,-25.9659],[61,-24.7863],[62,-24.2205],[64,-23.946],[67,-23.6407],[68,-23.496],[69,-23.2622],[70,-22.8817],[71,-21.929],[72,-21.2066],[73,-21.5497],[75,-22.8],[76,-23.2208],[77,-24.2863],[78,-24.8319],[81,-25.0019],[82,-25.0179],[83,-24.7821],[84,-25.4161],[85,-25.8262],[87,-26.3631],[88,-26.8785],[89,-27.1875],[90,-27.1894],[91,-26.3507],[92,-24.8246],[93,-25.112],[94,-24.5657],[96,-25.9454],[97,-25.1519],[98,-23.5216],[99,-21.0761],[100,-15.6347],[101,-32.303],[102,-21.1643],[103,-27.7007],[104,-18.9141],[105,-18.455],[106,-26.7574],[107,-25.7827],[108,-26.4237],[109,-26.2999],[110,-26.7953],[111,-28.0864],[112,-6.949],[113,-3.2305],[114,-1.3123],[115,0.6188],[116,1.7494],[117,2.9005],[118,3.8594],[119,4.686],[120,5.4113],[121,5.813],[122,6.3579],[123,6.8686],[124,7.2597],[125,7.1659],[126,7.369],[127,7.4777],[128,6.994],[129,7.0926],[130,7.2791],[131,7.2878],[132,7.2317],[133,7.3814],[134,7.7042],[135,7.9061],[136,8.2743],[137,8.8887],[138,9.4504],[139,9.2653],[140,9.3828],[141,9.3204],[142,9.2036],[144,9.1724],[146,9.0932],[147,9.0236],[148,9.0219],[149,8.9073],[150,8.9516],[151,8.8382],[152,8.7879],[154,8.628],[156,8.5813],[157,8.4365],[158,8.482],[159,8.3185],[160,8.1832],[161,8.2335],[162,8.2468],[163,8.391],[165,8.4009],[166,7.8483],[167,4.1673],[168,-3.6869],[169,-11.5229],[170,-18.3938],[171,-21.4023],[172,-22.1804],[173,-23.0542],[174,-23.0436],[175,-22.9085],[176,-23.489],[177,-23.3795],[178,-23.7911],[179,-24.3082],[180,-24.7467],[181,-25.4417],[182,-26.8413],[183,-27.5386],[184,-29.178]]
       }
       middle_03_r_ctrl.Scale.X {type=MovieSceneScriptingFloatChannel, default=1, const [1,185]}
       middle_03_r_ctrl.Scale.Y {type=MovieSceneScriptingFloatChannel, default=1, const [1,185]}
       middle_03_r_ctrl.Scale.Z {type=MovieSceneScriptingFloatChannel, default=1, const [1,185]}
       ring_metacarpal_r_ctrl.Location.X {
        type=MovieSceneScriptingFloatChannel
        default=0.0001
        thinned_from=185
        keys [[0,0.0001],[51,0.0001],[63,-0.0008],[70,0],[72,-0.0001],[80,0.0001],[82,0],[99,0.0001],[100,-0.0001],[101,-0.0001],[102,-0.0004],[104,0],[105,0],[106,-0.0008],[107,0.0001],[108,-0.0055],[109,-0.0007],[110,-0.0047],[111,0],[112,-0.0001],[113,-0.0012],[114,-0.0015],[115,-0.0005],[118,0],[135,0],[136,-0.001],[137,0.0001],[138,-0.0005],[139,0.0001],[140,-0.0001],[168,0.0001],[171,-0.0006],[172,0],[180,0],[182,-0.0003],[183,-0.0008],[184,0.0001]]
       }
       ring_metacarpal_r_ctrl.Location.Y {
        type=MovieSceneScriptingFloatChannel
        default=0.0002
        thinned_from=185
        keys [[0,0.0002],[51,0.0002],[52,0],[56,-0.0001],[61,-0.0009],[70,0.0002],[99,0.0002],[102,0.0005],[105,0.0003],[106,0.0004],[108,0.0002],[109,0.0003],[110,0.0012],[111,0.0006],[112,0.0011],[113,0.0006],[114,0.0009],[115,0.0005],[135,0.0002],[136,0.0005],[137,0.0002],[138,0.0006],[139,0.0003],[140,0.0004],[169,0.0003],[171,0.0007],[172,0.0003],[181,0.0003],[183,0.0005],[184,0.0002]]
       }
       ring_metacarpal_r_ctrl.Location.Z {
        type=MovieSceneScriptingFloatChannel
        default=0
        thinned_from=185
        keys [[0,0],[51,0],[63,-0.0007],[72,0],[101,0],[102,0.0004],[104,-0.0001],[105,-0.0001],[106,-0.0004],[107,0],[108,-0.0018],[109,-0.0003],[110,-0.0014],[111,0],[112,0.0001],[114,-0.0006],[115,-0.0002],[135,-0.0001],[136,-0.0006],[137,0],[138,-0.0004],[139,0],[171,-0.0002],[181,0.0001],[182,-0.0004],[183,-0.0004],[184,0]]
       }
       ring_metacarpal_r_ctrl.Rotation.X {type=MovieSceneScriptingFloatChannel, default=-0.04, const [-0.04,185]}
       ring_metacarpal_r_ctrl.Rotation.Y {type=MovieSceneScriptingFloatChannel, default=3.1532, const [3.1532,185]}
       ring_metacarpal_r_ctrl.Rotation.Z {type=MovieSceneScriptingFloatChannel, default=-0.8, const [-0.8,185]}
       ring_metacarpal_r_ctrl.Scale.X {type=MovieSceneScriptingFloatChannel, default=1, const [1,185]}
       ring_metacarpal_r_ctrl.Scale.Y {type=MovieSceneScriptingFloatChannel, default=1, const [1,185]}
       ring_metacarpal_r_ctrl.Scale.Z {type=MovieSceneScriptingFloatChannel, default=1, const [1,185]}
       ring_01_r_ctrl.Location.X {
        type=MovieSceneScriptingFloatChannel
        default=-0.0003
        thinned_from=185
        keys [[0,-0.0002],[55,-0.0007],[61,-0.0021],[63,-0.0022],[68,-0.001],[70,-0.0003],[72,-0.0009],[74,-0.0002],[82,-0.0005],[83,-0.0002],[99,-0.0004],[100,-0.001],[101,-0.0008],[102,-0.0025],[103,-0.0012],[104,-0.0004],[105,-0.0002],[106,-0.0026],[107,-0.0002],[108,-0.0154],[109,-0.0024],[110,-0.0133],[111,-0.0011],[112,-0.0026],[114,-0.005],[115,-0.002],[116,-0.0015],[117,-0.0008],[118,-0.0004],[133,-0.0002],[134,-0.0004],[135,-0.0002],[136,-0.0035],[137,-0.0002],[138,-0.0022],[139,-0.0003],[140,-0.0007],[141,-0.0002],[165,-0.0002],[166,-0.0004],[167,-0.0002],[169,-0.0009],[170,-0.0022],[171,-0.0027],[172,-0.0004],[180,-0.0003],[181,-0.0009],[182,-0.0013],[183,-0.0026],[184,-0.0003]]
       }
       ring_01_r_ctrl.Location.Y {
        type=MovieSceneScriptingFloatChannel
        default=0
        thinned_from=185
        keys [[0,0],[49,-0.0003],[55,-0.0011],[56,-0.0014],[60,-0.0037],[61,-0.0041],[62,-0.0043],[63,-0.0041],[65,-0.0032],[69,-0.0006],[70,-0.0002],[72,-0.0004],[74,0],[97,0],[99,0.0001],[100,-0.0001],[101,0],[102,-0.0005],[103,-0.0001],[105,0],[106,-0.0002],[107,0],[108,-0.0041],[109,-0.0005],[110,-0.0013],[111,0.0004],[112,0.0013],[113,-0.0012],[114,-0.0012],[115,-0.0003],[135,0],[136,-0.0005],[138,0.0003],[139,0],[140,0.0002],[169,0],[180,0],[181,-0.0002],[182,0.0002],[183,0],[184,0]]
       }
       ring_01_r_ctrl.Location.Z {
        type=MovieSceneScriptingFloatChannel
        default=0
        thinned_from=185
        keys [[0,0],[56,0.0005],[61,0.0015],[64,0.0014],[70,0.0001],[72,0.0004],[74,0],[82,-0.0001],[99,0.0001],[100,0.0003],[101,0.0002],[102,0.0023],[103,0.0009],[104,-0.0001],[105,0],[106,-0.0005],[107,0],[108,-0.0025],[109,-0.0006],[110,-0.0025],[111,0.0012],[112,0.0028],[113,0.0001],[114,-0.0003],[115,0.0002],[135,0],[136,-0.0003],[137,0],[138,-0.0003],[139,0],[169,0],[171,0.0005],[172,0.0001],[180,0.0001],[181,0.0004],[182,-0.0005],[183,-0.0006],[184,0]]
       }
       ring_01_r_ctrl.Rotation.X {
        type=MovieSceneScriptingFloatChannel
        default=-9.0969
        thinned_from=185
        keys [[0,-8.7246],[8,-8.7467],[9,-8.7666],[10,-8.7526],[12,-8.7662],[13,-8.7479],[14,-8.7695],[16,-8.7322],[17,-8.7464],[18,-8.7222],[20,-8.7335],[21,-8.7176],[24,-8.7518],[25,-8.7356],[26,-8.7527],[28,-8.7399],[29,-8.7636],[32,-8.7544],[33,-8.7745],[36,-8.7427],[37,-8.7657],[38,-8.739],[39,-8.7571],[41,-8.7474],[55,-8.8355],[56,-8.8594],[61,-9.116],[67,-9.2006],[68,-9.203],[69,-9.0897],[70,-8.8406],[71,-7.3563],[72,-6.0078],[74,-6.0115],[75,-5.9012],[76,-5.7285],[77,-5.4654],[78,-5.3423],[80,-5.2736],[83,-5.1854],[84,-5.0449],[85,-5.0795],[86,-5.0525],[87,-5.0453],[89,-5.0668],[90,-5.0003],[91,-5.0631],[93,-5.0912],[94,-5.2169],[95,-5.1249],[96,-5.2396],[97,-5.225],[98,-5.2801],[99,-5.929],[100,-7.9381],[101,-7.1432],[102,-8.081],[103,-6.1802],[104,-7.9029],[105,-7.8671],[106,-6.9556],[107,-7.3001],[108,-7.0009],[109,-7.0673],[110,-6.7947],[111,-7.1092],[112,-9.4187],[113,-9.9255],[114,-10.1807],[115,-10.6529],[116,-10.8126],[117,-10.8964],[124,-11.021],[125,-10.9707],[127,-10.9932],[131,-10.9599],[132,-10.9676],[133,-10.9993],[136,-11.2001],[138,-11.4493],[139,-11.5142],[140,-11.5316],[141,-11.5932],[142,-11.5653],[150,-11.6281],[152,-11.5971],[154,-11.6273],[159,-11.5933],[161,-11.5651],[165,-11.5892],[166,-11.5571],[167,-11.3507],[168,-10.8859],[169,-10.1953],[170,-9.7262],[171,-9.5049],[172,-9.4385],[173,-9.4924],[174,-9.4594],[175,-9.4677],[176,-9.4397],[177,-9.4475],[178,-9.39],[179,-9.3981],[180,-9.3357],[181,-9.3215],[182,-9.1117],[183,-9.1697],[184,-9.0969]]
       }
       ring_01_r_ctrl.Rotation.Y {
        type=MovieSceneScriptingFloatChannel
        default=-1.9774
        thinned_from=185
        keys [[0,-2.2486],[1,-2.2332],[2,-2.2997],[3,-2.2502],[4,-2.2408],[5,-2.2099],[6,-2.3458],[7,-2.2031],[8,-2.189],[9,-2.1604],[10,-2.1241],[11,-2.1237],[12,-2.2029],[13,-2.203],[14,-2.2666],[15,-2.2984],[16,-2.1994],[17,-2.3124],[18,-2.2024],[19,-2.2154],[20,-2.2821],[21,-2.2732],[22,-2.298],[23,-2.4125],[24,-2.3338],[25,-2.2266],[26,-2.3716],[27,-2.2896],[28,-2.1908],[29,-2.3619],[30,-2.2499],[31,-2.1758],[32,-2.2531],[33,-2.1812],[34,-2.315],[35,-2.1694],[36,-2.1557],[37,-2.2658],[38,-2.2027],[39,-2.2981],[40,-2.2568],[42,-2.2078],[43,-2.1755],[44,-2.2022],[45,-2.1943],[46,-2.1718],[47,-2.1359],[48,-2.0882],[49,-2.0303],[50,-1.9635],[51,-1.8893],[52,-1.8092],[54,-1.6373],[55,-1.5484],[56,-1.3807],[57,-1.0918],[58,-0.7353],[59,-0.3646],[60,-0.0336],[61,0.2044],[62,0.2956],[63,0.2921],[64,0.2834],[66,0.2606],[67,0.2519],[68,0.2484],[69,0.2951],[70,0.3978],[71,2.024],[72,3.5943],[73,3.5783],[74,3.5433],[75,3.3591],[76,3.3516],[77,3.402],[78,3.5396],[79,3.535],[82,3.4856],[83,3.1691],[84,3.0796],[85,3.0286],[86,2.9111],[87,2.8932],[88,2.8907],[89,2.8204],[90,2.78],[91,2.863],[92,2.9248],[93,2.8695],[94,2.8558],[95,2.7965],[96,2.7051],[97,2.3565],[98,2.1103],[99,1.9611],[100,-0.1652],[101,-0.9711],[102,-1.235],[103,-1.1297],[104,0.0029],[105,0.2107],[106,-3.3584],[107,-3.6345],[108,-3.3021],[109,-3.3682],[110,-3.3068],[111,-3.1199],[112,-3.6076],[113,-4.2256],[114,-4.7059],[115,-5.9954],[116,-6.5227],[117,-6.9723],[118,-7.0913],[119,-7.3603],[120,-7.5493],[121,-7.7539],[122,-7.8639],[123,-8.0218],[124,-8.4914],[125,-8.3261],[126,-8.2614],[127,-8.2757],[128,-8.3344],[129,-8.3727],[130,-8.334],[131,-8.3812],[132,-8.3863],[133,-8.2631],[134,-8.0817],[135,-8.0163],[136,-8.1121],[137,-8.3135],[138,-8.6677],[139,-8.9607],[140,-9.0984],[141,-9.1776],[142,-9.1016],[143,-9.1966],[144,-9.131],[147,-9.1707],[148,-9.1567],[149,-9.17],[150,-9.2622],[151,-9.2456],[152,-9.2034],[153,-9.2739],[154,-9.2995],[155,-9.2094],[156,-9.2116],[157,-9.1942],[158,-9.2099],[159,-9.181],[160,-9.0927],[161,-9.1036],[162,-9.1356],[165,-9.1186],[166,-8.9102],[167,-7.4527],[168,-3.6423],[169,-1.8489],[170,-1.5006],[171,-1.8104],[172,-1.8987],[173,-2.3123],[174,-2.2391],[175,-2.3422],[176,-2.3809],[177,-2.3886],[178,-2.167],[179,-2.3742],[180,-2.2416],[181,-2.2494],[182,-0.9581],[183,-1.2234],[184,-1.9774]]
       }
       ring_01_r_ctrl.Rotation.Z {
        type=MovieSceneScriptingFloatChannel
        default=-61.5875
        thinned_from=185
        keys [[0,-62.9236],[2,-62.5711],[3,-62.7782],[4,-62.4358],[5,-62.9044],[8,-62.6003],[10,-62.6292],[11,-62.743],[12,-62.5536],[15,-62.9886],[16,-63.3258],[18,-63.4764],[20,-63.2813],[21,-63.5084],[24,-63.6111],[25,-63.8798],[26,-63.4974],[28,-63.5956],[29,-63.2442],[30,-63.3851],[31,-63.6758],[32,-63.5577],[33,-63.6046],[34,-63.2752],[35,-63.5894],[36,-63.645],[37,-63.5584],[41,-63.8022],[42,-63.6564],[44,-63.9745],[45,-63.8957],[46,-63.6691],[48,-62.8316],[50,-61.5803],[54,-58.3093],[55,-57.4176],[56,-55.6954],[57,-52.6799],[58,-48.9149],[59,-44.9445],[60,-41.3128],[61,-38.5637],[62,-37.2413],[64,-36.4175],[66,-35.7838],[67,-35.5977],[68,-35.5304],[70,-35.973],[71,-35.7123],[72,-35.3402],[74,-35.0432],[75,-34.5725],[76,-33.9974],[77,-33.2169],[78,-32.2945],[82,-32.3263],[83,-33.8011],[84,-34.1995],[85,-34.2796],[86,-33.7415],[88,-34.5353],[90,-34.2603],[91,-34.2567],[92,-34.0961],[93,-34.3556],[94,-35.068],[95,-34.7509],[97,-34.5604],[98,-34.0479],[99,-32.2391],[100,-34.4884],[101,-14.9758],[102,-8.5553],[103,-11.66],[104,-19.5903],[105,-20.3505],[106,-23.1484],[107,-23.506],[108,-21.3991],[109,-22.2514],[110,-21.4352],[111,-22.0719],[112,-45.0244],[113,-50.5644],[115,-55.6323],[116,-56.6081],[117,-56.9138],[118,-56.565],[119,-56.7009],[120,-57.188],[123,-57.5104],[124,-57.7601],[125,-57.4487],[127,-57.4467],[129,-57.8376],[130,-57.6858],[132,-57.686],[134,-57.3331],[135,-57.7063],[136,-58.2419],[137,-58.497],[139,-60.9208],[141,-61.7909],[143,-62.084],[144,-62.1047],[145,-61.9358],[152,-62.2295],[153,-62.503],[154,-62.9152],[155,-62.7139],[160,-62.3304],[162,-62.502],[163,-62.3022],[164,-62.557],[165,-62.4682],[166,-62.1255],[167,-60.2617],[168,-54.9345],[170,-60.1825],[171,-66.1057],[173,-70.0952],[175,-70.2473],[176,-69.9119],[177,-70.0894],[178,-69.4906],[179,-69.4368],[180,-69.2345],[181,-68.069],[182,-57.335],[183,-57.5125],[184,-61.5875]]
       }
       ring_01_r_ctrl.Scale.X {type=MovieSceneScriptingFloatChannel, default=1, const [1,185]}
       ring_01_r_ctrl.Scale.Y {type=MovieSceneScriptingFloatChannel, default=1, const [1,185]}
       ring_01_r_ctrl.Scale.Z {type=MovieSceneScriptingFloatChannel, default=1, const [1,185]}
       ring_02_r_ctrl.Location.X {
        type=MovieSceneScriptingFloatChannel
        default=0.0003
        thinned_from=185
        keys [[0,0.0003],[68,0],[70,0.0002],[72,0],[82,0.0001],[97,0.0002],[99,-0.0001],[100,-0.0068],[101,-0.0061],[102,-0.0005],[103,-0.0001],[105,0.0002],[106,-0.0005],[107,0.0003],[108,-0.0047],[109,-0.0007],[110,-0.006],[111,-0.003],[112,-0.0068],[113,0.0002],[133,0.0003],[138,0.0001],[169,0.0002],[171,-0.0002],[172,0.0002],[180,0.0002],[181,-0.0003],[182,-0.0004],[183,0.0001],[184,0.0003]]
       }
       ring_02_r_ctrl.Location.Y {
        type=MovieSceneScriptingFloatChannel
        default=0
        thinned_from=185
        keys [[0,0],[82,-0.0001],[99,-0.0001],[100,-0.0073],[101,-0.0043],[102,-0.0001],[105,0],[106,-0.0008],[107,0],[108,-0.0054],[109,-0.0009],[110,-0.0044],[111,-0.0014],[112,-0.0047],[113,-0.0008],[114,-0.0011],[115,-0.0005],[117,-0.0001],[133,0],[134,-0.0001],[135,0],[136,-0.0008],[137,0],[138,-0.0007],[139,0],[140,-0.0002],[141,0],[168,0],[169,-0.0003],[170,-0.0009],[171,-0.0012],[172,-0.0001],[180,0],[181,-0.0005],[182,-0.0007],[183,-0.0005],[184,0]]
       }
       ring_02_r_ctrl.Location.Z {
        type=MovieSceneScriptingFloatChannel
        default=0
        thinned_from=185
        keys [[0,0],[68,0.0005],[70,0.0001],[72,0.0002],[99,0.0001],[100,0.0004],[101,-0.0002],[102,0.0004],[104,0],[106,0.0001],[107,0],[108,0.001],[109,-0.0001],[110,0.0003],[111,-0.0013],[112,-0.0039],[113,-0.0003],[114,-0.0005],[115,-0.0001],[135,0],[136,0.0005],[137,0],[138,0.0002],[139,0],[168,0],[171,-0.0005],[172,0],[181,0],[182,0.0002],[184,0]]
       }
       ring_02_r_ctrl.Rotation.X {
        type=MovieSceneScriptingFloatChannel
        default=-10.3618
        thinned_from=185
        keys [[0,-12.1333],[2,-12.1719],[3,-12.1475],[4,-12.1779],[5,-12.1151],[6,-12.1633],[7,-12.1347],[8,-12.1564],[11,-12.1088],[12,-12.1772],[13,-12.1466],[14,-12.1726],[15,-12.1534],[16,-12.0797],[17,-12.104],[18,-12.0307],[19,-12.0797],[21,-12.0379],[23,-12.0441],[24,-12.0138],[25,-11.9417],[26,-12.0168],[28,-11.9433],[29,-12.0102],[30,-11.9817],[31,-11.9141],[32,-11.9709],[33,-11.9515],[34,-12.0182],[35,-11.9273],[37,-11.9593],[38,-11.9413],[39,-11.9625],[41,-11.9379],[42,-11.9636],[43,-11.909],[45,-11.9027],[47,-11.8401],[50,-11.6552],[55,-11.2102],[56,-11.0299],[57,-10.7188],[59,-9.9338],[60,-9.575],[61,-9.3155],[62,-9.2128],[68,-9.192],[70,-9.0672],[72,-9.7862],[74,-9.8356],[75,-9.9252],[76,-9.9929],[77,-10.1755],[78,-10.2985],[83,-10.3563],[84,-10.4755],[85,-10.5374],[86,-10.5541],[87,-10.5936],[92,-10.6976],[93,-10.6636],[94,-10.5687],[95,-10.6888],[96,-10.6544],[97,-10.5275],[98,-10.3138],[99,-10.0298],[100,-7.4968],[101,-11.2196],[102,-12.0506],[103,-10.1221],[104,-7.4441],[105,-6.9323],[106,-10.5457],[107,-10.4708],[108,-10.682],[109,-10.7601],[110,-10.9191],[111,-10.83],[112,-5.6215],[113,-5.0123],[114,-4.9394],[115,-5.4832],[116,-5.5198],[117,-5.3806],[118,-5.3275],[120,-5.0801],[121,-5.0437],[123,-4.9275],[124,-4.8822],[125,-4.9051],[126,-4.8553],[127,-4.8481],[129,-4.8581],[131,-4.8525],[132,-4.8358],[133,-4.8735],[134,-4.8256],[135,-4.8117],[136,-4.8151],[137,-4.7648],[138,-4.7301],[139,-4.8011],[140,-4.7548],[141,-4.7707],[142,-4.7471],[144,-4.7526],[145,-4.815],[146,-4.8338],[148,-4.8137],[150,-4.8566],[151,-4.8442],[152,-4.8726],[153,-4.8674],[154,-4.8509],[155,-4.8633],[156,-4.8362],[158,-4.8574],[159,-4.888],[160,-4.8997],[162,-4.8765],[163,-4.9317],[164,-4.8894],[165,-4.8833],[166,-4.9464],[167,-5.451],[168,-6.4477],[169,-7.0663],[170,-8.0465],[171,-8.5591],[172,-8.7001],[174,-8.7925],[175,-8.7888],[176,-8.8801],[177,-8.8643],[178,-9.1622],[179,-9.2475],[180,-9.4234],[181,-9.5532],[182,-9.8452],[183,-10.1647],[184,-10.3618]]
       }
       ring_02_r_ctrl.Rotation.Y {
        type=MovieSceneScriptingFloatChannel
        default=6.5724
        thinned_from=185
        keys [[0,5.7278],[2,5.703],[3,5.7185],[4,5.6991],[5,5.7397],[6,5.7086],[7,5.727],[8,5.713],[9,5.7336],[11,5.7435],[12,5.6996],[13,5.7193],[14,5.7024],[15,5.7149],[16,5.7619],[17,5.7465],[18,5.7926],[19,5.7619],[20,5.7685],[21,5.788],[23,5.7841],[24,5.8028],[25,5.8467],[26,5.801],[28,5.8458],[29,5.8051],[30,5.8225],[31,5.8632],[32,5.8291],[33,5.8408],[34,5.8002],[35,5.8555],[37,5.8362],[38,5.8469],[39,5.8342],[41,5.849],[42,5.8335],[43,5.8662],[46,5.8771],[50,5.9599],[55,6.1249],[56,6.1916],[57,6.3067],[59,6.5966],[60,6.7288],[61,6.8241],[62,6.8612],[68,6.8645],[70,6.8849],[72,6.7401],[74,6.7276],[76,6.6853],[77,6.6312],[78,6.5935],[83,6.5747],[84,6.5343],[85,6.5122],[92,6.4528],[93,6.4657],[94,6.5009],[95,6.4564],[96,6.469],[97,6.5153],[98,6.5871],[99,6.6731],[100,6.9967],[101,6.2522],[102,5.7645],[103,6.6411],[104,6.9698],[105,6.9237],[106,6.5106],[107,6.5357],[108,6.4555],[109,6.4276],[110,6.3587],[111,6.3587],[112,6.5068],[113,6.4488],[114,6.4201],[115,6.6141],[116,6.6254],[117,6.5805],[118,6.563],[120,6.4752],[124,6.3983],[125,6.4075],[126,6.3874],[132,6.3793],[133,6.3948],[134,6.3753],[136,6.3717],[138,6.3334],[139,6.365],[140,6.346],[141,6.3522],[144,6.3445],[145,6.3707],[148,6.3702],[150,6.3879],[151,6.3828],[152,6.3944],[155,6.3906],[156,6.3795],[160,6.4054],[162,6.396],[163,6.4181],[164,6.4012],[165,6.3987],[166,6.4207],[167,6.6035],[168,6.8486],[169,6.9366],[170,6.9687],[171,6.9426],[177,6.9134],[178,6.8698],[179,6.8551],[181,6.795],[182,6.7242],[183,6.6363],[184,6.5724]]
       }
       ring_02_r_ctrl.Rotation.Z {
        type=MovieSceneScriptingFloatChannel
        default=-60.1393
        thinned_from=185
        keys [[0,-74.4145],[2,-74.7476],[3,-74.5375],[4,-74.7997],[5,-74.2573],[6,-74.6727],[7,-74.4263],[8,-74.6134],[9,-74.3374],[11,-74.2034],[12,-74.7933],[13,-74.529],[14,-74.7536],[15,-74.5878],[16,-73.9541],[17,-74.1621],[18,-73.5349],[19,-73.9535],[20,-73.8633],[21,-73.5966],[23,-73.6495],[24,-73.3913],[25,-72.7793],[26,-73.4166],[28,-72.7924],[29,-73.3607],[30,-73.1185],[31,-72.5458],[32,-73.0262],[33,-72.8622],[34,-73.4286],[35,-72.6578],[37,-72.9276],[38,-72.7757],[39,-72.9553],[41,-72.7473],[42,-72.9652],[43,-72.5032],[45,-72.4554],[47,-71.9675],[49,-71.086],[53,-68.5357],[55,-67.0638],[56,-65.6612],[57,-63.2401],[58,-60.2486],[59,-57.1343],[60,-54.3453],[61,-52.3292],[62,-51.5342],[68,-51.3928],[70,-50.4591],[71,-53.049],[72,-55.779],[74,-56.1527],[75,-56.8238],[76,-57.3349],[77,-58.7139],[78,-59.6527],[83,-60.0941],[84,-61.0077],[85,-61.4843],[86,-61.6132],[87,-61.9181],[91,-62.6395],[92,-62.7247],[93,-62.4612],[94,-61.7258],[95,-62.6564],[96,-62.3882],[97,-61.4062],[98,-59.7648],[99,-57.6076],[100,-39.0689],[101,-66.8437],[102,-73.5631],[103,-58.2501],[104,-38.6503],[105,-34.9071],[106,-61.5436],[107,-60.9716],[108,-62.6152],[109,-63.2127],[110,-64.4664],[111,-63.4029],[112,-24.0085],[113,-20.5136],[114,-19.9469],[115,-24.1277],[116,-24.4065],[117,-23.3499],[118,-22.9423],[120,-21.0359],[121,-20.7544],[122,-20.3285],[123,-19.8477],[124,-19.4938],[125,-19.6736],[126,-19.2837],[127,-19.2268],[129,-19.3056],[131,-19.261],[132,-19.1302],[133,-19.426],[134,-19.0506],[135,-18.9414],[136,-18.9628],[137,-18.5719],[138,-18.3008],[139,-18.8575],[140,-18.4923],[141,-18.6184],[142,-18.4319],[143,-18.4332],[144,-18.4757],[145,-18.9667],[146,-19.1145],[148,-18.9567],[149,-19.1479],[150,-19.2936],[151,-19.1962],[152,-19.4189],[153,-19.3784],[154,-19.2488],[155,-19.3457],[156,-19.1333],[158,-19.2998],[159,-19.5397],[160,-19.6314],[161,-19.5751],[162,-19.4491],[163,-19.8815],[164,-19.5508],[165,-19.5025],[166,-19.9975],[167,-23.8842],[168,-31.3481],[169,-35.8907],[170,-43.0447],[171,-46.7619],[172,-47.7688],[174,-48.4428],[175,-48.4156],[176,-49.0837],[177,-48.9675],[178,-51.1524],[179,-51.783],[180,-53.0842],[181,-54.0456],[182,-56.2282],[183,-58.6284],[184,-60.1393]]
       }
       ring_02_r_ctrl.Scale.X {type=MovieSceneScriptingFloatChannel, default=1, const [1,185]}
       ring_02_r_ctrl.Scale.Y {type=MovieSceneScriptingFloatChannel, default=1, const [1,185]}
       ring_02_r_ctrl.Scale.Z {type=MovieSceneScriptingFloatChannel, default=1, const [1,185]}
       ring_03_r_ctrl.Location.X {
        type=MovieSceneScriptingFloatChannel
        default=-0.0001
        thinned_from=185
        keys [[0,0],[55,-0.0006],[61,-0.0022],[63,-0.0022],[70,-0.0002],[72,-0.0004],[74,0],[82,-0.0002],[97,-0.0001],[99,-0.0003],[101,-0.0001],[102,-0.0053],[103,-0.0023],[104,-0.0004],[105,-0.0001],[106,-0.0005],[107,0],[108,-0.0032],[109,-0.0005],[110,-0.0031],[111,-0.003],[112,-0.0079],[113,-0.0004],[114,-0.0006],[115,-0.0002],[135,0],[136,-0.0005],[137,0],[165,0],[166,-0.0002],[167,0],[169,-0.0004],[170,-0.0011],[171,-0.0013],[172,-0.0001],[180,0],[181,-0.0004],[183,-0.0004],[184,-0.0001]]
       }
       ring_03_r_ctrl.Location.Y {
        type=MovieSceneScriptingFloatChannel
        default=-0.0001
        thinned_from=185
        keys [[0,-0.0001],[56,-0.0004],[62,-0.0011],[70,-0.0001],[97,-0.0001],[99,-0.0003],[101,-0.0001],[102,-0.0026],[103,-0.0007],[104,-0.0003],[106,0],[108,0],[109,-0.0002],[110,-0.0001],[111,-0.0021],[112,-0.0033],[113,0],[135,-0.0001],[136,0.0002],[137,-0.0001],[138,0],[171,-0.0003],[180,-0.0001],[181,-0.0002],[183,0],[184,-0.0001]]
       }
       ring_03_r_ctrl.Location.Z {
        type=MovieSceneScriptingFloatChannel
        default=0
        thinned_from=185
        keys [[0,0],[97,0],[99,-0.0002],[100,0.0002],[101,0.0001],[102,-0.0011],[103,-0.0006],[104,0.0002],[105,0],[106,0.0004],[107,0],[108,0.002],[109,0.0004],[110,0.002],[111,-0.0009],[112,-0.0032],[113,-0.0004],[114,-0.0005],[115,-0.0001],[135,0],[136,0.0003],[137,0],[166,-0.0001],[168,-0.0001],[169,-0.0002],[170,-0.0006],[171,-0.0007],[172,0],[180,0],[181,-0.0002],[182,0.0003],[184,0]]
       }
       ring_03_r_ctrl.Rotation.X {
        type=MovieSceneScriptingFloatChannel
        default=-1.8407
        thinned_from=185
        keys [[0,-2.3808],[1,-2.3939],[2,-2.3864],[3,-2.3968],[4,-2.4132],[5,-2.3888],[6,-2.3751],[7,-2.398],[9,-2.3966],[10,-2.4005],[11,-2.3839],[12,-2.397],[14,-2.3847],[15,-2.3972],[16,-2.3798],[17,-2.3454],[18,-2.3812],[19,-2.3698],[20,-2.3838],[21,-2.3719],[22,-2.3751],[23,-2.3534],[25,-2.3555],[27,-2.3389],[29,-2.3428],[30,-2.323],[31,-2.3376],[32,-2.3468],[34,-2.3375],[35,-2.3421],[38,-2.3375],[39,-2.3407],[40,-2.3511],[41,-2.3311],[42,-2.3632],[53,-2.2485],[55,-2.2195],[56,-2.1892],[57,-2.1366],[58,-2.0701],[60,-1.9288],[61,-1.8707],[62,-1.832],[63,-1.8095],[67,-1.7484],[68,-1.7276],[69,-1.6941],[70,-1.6398],[71,-1.4139],[72,-1.2209],[73,-1.2448],[74,-1.2974],[75,-1.4181],[76,-1.4869],[77,-1.5829],[78,-1.6071],[82,-1.6081],[83,-1.6301],[84,-1.7077],[85,-1.7672],[86,-1.8425],[87,-1.8651],[88,-1.8764],[89,-1.9163],[90,-1.9256],[91,-1.8942],[92,-1.847],[93,-1.8683],[94,-1.7951],[95,-1.9095],[96,-1.9512],[97,-1.8961],[98,-1.8106],[99,-1.6638],[100,-1.1406],[101,-2.5095],[102,-2.7824],[103,-1.2953],[104,-1.1355],[105,-1.1488],[106,-3.3207],[107,-3.4678],[108,-3.4341],[109,-3.4317],[110,-3.3513],[111,-3.2926],[112,-1.0149],[113,-0.864],[114,-0.8157],[115,-0.8078],[116,-0.7799],[118,-0.7449],[119,-0.7199],[120,-0.7158],[121,-0.7083],[122,-0.7027],[123,-0.6948],[124,-0.6922],[125,-0.698],[126,-0.7],[127,-0.6978],[128,-0.7068],[129,-0.6942],[130,-0.6909],[131,-0.684],[132,-0.6856],[133,-0.6799],[134,-0.6832],[135,-0.6884],[136,-0.6827],[137,-0.7119],[138,-0.724],[139,-0.7147],[140,-0.7177],[141,-0.717],[142,-0.7201],[143,-0.7112],[144,-0.7168],[145,-0.7144],[146,-0.7085],[147,-0.7116],[148,-0.7071],[149,-0.7143],[150,-0.7111],[151,-0.7023],[152,-0.7066],[153,-0.706],[155,-0.6997],[156,-0.6986],[157,-0.7005],[158,-0.7003],[159,-0.7117],[160,-0.7119],[161,-0.7056],[162,-0.7119],[164,-0.6961],[165,-0.7076],[166,-0.6954],[167,-0.7269],[168,-0.7962],[169,-0.8992],[170,-1.1565],[171,-1.3681],[172,-1.4565],[173,-1.5827],[174,-1.6117],[175,-1.5962],[176,-1.6386],[177,-1.6214],[178,-1.5945],[179,-1.6561],[180,-1.643],[181,-1.694],[182,-1.7727],[183,-1.737],[184,-1.8407]]
       }
       ring_03_r_ctrl.Rotation.Y {
        type=MovieSceneScriptingFloatChannel
        default=4.4037
        thinned_from=185
        keys [[0,4.9799],[1,4.9926],[2,4.9853],[4,5.0111],[5,4.9877],[6,4.9744],[7,4.9964],[10,4.9988],[11,4.9829],[12,4.9956],[14,4.9836],[15,4.9957],[16,4.979],[17,4.9455],[18,4.9803],[19,4.9693],[20,4.9828],[21,4.9713],[22,4.9745],[23,4.9533],[25,4.9553],[27,4.9391],[29,4.943],[30,4.9235],[32,4.9469],[34,4.9377],[39,4.9409],[40,4.9511],[41,4.9315],[42,4.9629],[55,4.8128],[56,4.7807],[57,4.725],[60,4.5028],[61,4.4392],[62,4.3954],[64,4.3487],[67,4.2965],[68,4.2708],[69,4.2276],[70,4.1564],[71,3.8373],[72,3.5611],[73,3.5985],[74,3.6808],[75,3.8561],[76,3.9526],[77,4.081],[78,4.1133],[82,4.1146],[83,4.1432],[84,4.2416],[85,4.3149],[86,4.4052],[87,4.4317],[88,4.4449],[89,4.491],[90,4.5016],[91,4.4655],[92,4.4104],[93,4.4356],[94,4.3487],[95,4.4834],[96,4.5307],[97,4.4675],[98,4.3662],[99,4.185],[100,3.4415],[101,5.1067],[102,5.3106],[103,3.6625],[104,3.4273],[105,3.449],[106,5.7581],[107,5.8579],[108,5.8351],[109,5.8334],[110,5.7777],[111,5.685],[112,3.043],[114,2.8548],[115,2.838],[116,2.7815],[118,2.7092],[119,2.6562],[120,2.6474],[123,2.6022],[124,2.5965],[125,2.6092],[126,2.6134],[127,2.6087],[128,2.6282],[129,2.6008],[130,2.5937],[131,2.5787],[132,2.5821],[133,2.5695],[135,2.5881],[136,2.5773],[137,2.6391],[138,2.6647],[139,2.6451],[140,2.6514],[141,2.65],[142,2.6567],[143,2.6374],[144,2.6495],[145,2.6445],[146,2.6317],[147,2.6385],[148,2.6287],[149,2.6443],[150,2.6374],[151,2.6185],[152,2.6276],[156,2.6105],[158,2.614],[159,2.6385],[160,2.6391],[161,2.6256],[162,2.6391],[164,2.6051],[165,2.6298],[166,2.6034],[167,2.671],[168,2.8143],[169,3.0147],[170,3.4656],[171,3.7887],[172,3.911],[173,4.0819],[174,4.1192],[175,4.0992],[176,4.1539],[177,4.132],[178,4.097],[179,4.1765],[180,4.1597],[181,4.2244],[182,4.3224],[183,4.2765],[184,4.4037]]
       }
       ring_03_r_ctrl.Rotation.Z {
        type=MovieSceneScriptingFloatChannel
        default=-19.679
        thinned_from=185
        keys [[0,-25.3968],[1,-25.5289],[2,-25.4534],[3,-25.5583],[4,-25.7223],[5,-25.4771],[6,-25.3392],[7,-25.5697],[9,-25.5562],[10,-25.5944],[11,-25.4282],[12,-25.56],[14,-25.4353],[15,-25.5618],[16,-25.3878],[17,-25.0388],[18,-25.4009],[19,-25.2859],[20,-25.4266],[21,-25.3075],[22,-25.3399],[23,-25.1207],[25,-25.1413],[27,-24.9736],[29,-25.013],[30,-24.8123],[31,-24.9599],[32,-25.0533],[34,-24.9591],[35,-25.0057],[38,-24.959],[39,-24.9916],[40,-25.097],[41,-24.8943],[42,-25.2192],[53,-24.0207],[55,-23.7143],[56,-23.3951],[57,-22.8407],[58,-22.1381],[60,-20.6376],[61,-20.014],[62,-19.5911],[63,-19.3388],[67,-18.656],[68,-18.4173],[69,-18.0219],[70,-17.3738],[71,-14.5473],[72,-12.1121],[73,-12.4326],[74,-13.1377],[75,-14.6777],[76,-15.5329],[77,-16.6969],[78,-16.9828],[82,-16.9947],[83,-17.2542],[84,-18.1602],[85,-18.8439],[86,-19.6947],[87,-19.9473],[88,-20.0729],[89,-20.514],[90,-20.6159],[91,-20.2699],[92,-19.7451],[93,-19.9833],[94,-19.1611],[95,-20.4373],[96,-20.8969],[97,-20.2912],[98,-19.3396],[99,-17.6551],[100,-10.9562],[101,-26.6212],[102,-29.3306],[103,-13.0943],[104,-10.966],[105,-11.149],[106,-34.2686],[107,-35.5373],[108,-35.2414],[109,-35.2168],[110,-34.4904],[111,-33.951],[112,-8.7095],[113,-7.0022],[114,-6.2444],[115,-6.1033],[116,-5.647],[117,-5.3611],[118,-5.0641],[119,-4.6407],[120,-4.5702],[121,-4.4433],[122,-4.3462],[123,-4.211],[124,-4.1659],[125,-4.2667],[126,-4.2996],[127,-4.2626],[128,-4.4174],[129,-4.1999],[130,-4.1439],[131,-4.0245],[132,-4.0518],[133,-3.9519],[134,-4.0092],[135,-4.0996],[136,-4.0227],[137,-4.5041],[138,-4.7052],[139,-4.552],[140,-4.6028],[141,-4.5912],[142,-4.6447],[143,-4.4914],[144,-4.5868],[145,-4.5472],[146,-4.4457],[147,-4.4995],[148,-4.4213],[149,-4.5458],[150,-4.491],[151,-4.3402],[152,-4.4131],[153,-4.4028],[155,-4.2947],[156,-4.2766],[157,-4.309],[158,-4.3046],[159,-4.4997],[160,-4.5044],[161,-4.3969],[162,-4.5039],[163,-4.3623],[164,-4.2342],[165,-4.4304],[166,-4.2242],[167,-4.7585],[168,-5.9081],[169,-7.5404],[170,-11.3094],[171,-14.0892],[172,-15.1617],[173,-16.6975],[174,-17.0367],[175,-16.8542],[176,-17.3555],[177,-17.1522],[178,-16.8337],[179,-17.561],[180,-17.4066],[181,-18.0025],[182,-18.922],[183,-18.4849],[184,-19.679]]
       }
       ring_03_r_ctrl.Scale.X {type=MovieSceneScriptingFloatChannel, default=1, const [1,185]}
       ring_03_r_ctrl.Scale.Y {type=MovieSceneScriptingFloatChannel, default=1, const [1,185]}
       ring_03_r_ctrl.Scale.Z {type=MovieSceneScriptingFloatChannel, default=1, const [1,185]}
       pinky_metacarpal_r_ctrl.Location.X {
        type=MovieSceneScriptingFloatChannel
        default=-0.0001
        thinned_from=185
        keys [[0,-0.0001],[78,-0.0001],[82,-0.0002],[101,-0.0002],[102,0.0002],[104,-0.0002],[105,-0.0001],[106,-0.0011],[107,-0.0001],[108,-0.0061],[109,-0.0011],[110,-0.0059],[111,0.0002],[112,0.0007],[113,-0.0012],[114,-0.0014],[115,-0.0006],[118,-0.0002],[133,-0.0001],[135,-0.0001],[136,-0.0012],[137,-0.0001],[138,-0.0008],[139,-0.0001],[140,-0.0003],[168,-0.0001],[171,-0.0007],[172,-0.0001],[181,-0.0002],[183,-0.0012],[184,-0.0001]]
       }
       pinky_metacarpal_r_ctrl.Location.Y {
        type=MovieSceneScriptingFloatChannel
        default=0.0001
        thinned_from=185
        keys [[0,0.0001],[48,0.0001],[62,-0.0012],[70,0],[99,0.0001],[101,0.0003],[102,0.0006],[104,0.0001],[106,0],[107,0],[108,-0.0016],[109,-0.0001],[110,0],[112,0.0014],[113,0.0012],[114,0.0016],[115,0.0006],[116,0.0004],[117,0.0001],[135,0.0001],[136,0.0002],[137,0],[138,0.0003],[139,0.0001],[140,0.0002],[168,0],[169,0.0002],[170,0.0008],[171,0.0009],[172,0.0002],[180,0],[184,0.0001]]
       }
       pinky_metacarpal_r_ctrl.Location.Z {
        type=MovieSceneScriptingFloatChannel
        default=0
        thinned_from=185
        keys [[0,0],[48,0],[56,-0.001],[61,-0.0028],[62,-0.0029],[64,-0.0025],[69,-0.0003],[70,-0.0001],[97,0],[101,0],[102,-0.0002],[105,0],[106,-0.0008],[107,0],[108,-0.0034],[109,-0.0005],[110,-0.0026],[111,-0.0007],[112,-0.0019],[113,-0.001],[114,-0.0015],[115,-0.0006],[118,-0.0001],[133,0],[134,-0.0002],[135,0],[136,-0.0015],[137,0],[138,-0.001],[139,-0.0001],[140,-0.0003],[141,0],[168,0],[170,-0.0007],[171,-0.0008],[172,-0.0001],[181,0],[182,-0.0006],[183,-0.0008],[184,0]]
       }
       pinky_metacarpal_r_ctrl.Rotation.X {type=MovieSceneScriptingFloatChannel, default=-0.58, const [-0.58,185]}
       pinky_metacarpal_r_ctrl.Rotation.Y {type=MovieSceneScriptingFloatChannel, default=7.1566, const [7.1566,185]}
       pinky_metacarpal_r_ctrl.Rotation.Z {type=MovieSceneScriptingFloatChannel, default=-4.63, const [-4.63,185]}
       pinky_metacarpal_r_ctrl.Scale.X {type=MovieSceneScriptingFloatChannel, default=1, const [1,185]}
       pinky_metacarpal_r_ctrl.Scale.Y {type=MovieSceneScriptingFloatChannel, default=1, const [1,185]}
       pinky_metacarpal_r_ctrl.Scale.Z {type=MovieSceneScriptingFloatChannel, default=1, const [1,185]}
       pinky_01_r_ctrl.Location.X {
        type=MovieSceneScriptingFloatChannel
        default=0
        thinned_from=185
        keys [[0,0.0001],[56,-0.0003],[61,-0.0012],[68,-0.0005],[70,0],[72,-0.0004],[74,0.0001],[82,-0.0002],[83,0.0001],[97,0.0001],[99,-0.0001],[100,-0.0006],[101,-0.0004],[102,-0.0018],[103,-0.0007],[104,-0.0002],[105,0.0001],[106,-0.0023],[107,0.0001],[108,-0.0148],[109,-0.0021],[110,-0.0132],[111,-0.0007],[114,-0.0047],[115,-0.0016],[116,-0.0012],[117,-0.0004],[118,-0.0001],[133,0.0001],[134,-0.0001],[135,0.0001],[136,-0.0032],[137,0.0001],[138,-0.0019],[139,0],[140,-0.0004],[141,0.0001],[165,0.0001],[166,-0.0001],[167,0.0001],[169,-0.0005],[170,-0.0019],[171,-0.0023],[172,-0.0001],[180,0.0001],[182,-0.0011],[183,-0.0024],[184,0]]
       }
       pinky_01_r_ctrl.Location.Y {
        type=MovieSceneScriptingFloatChannel
        default=0.0002
        thinned_from=185
        keys [[0,0.0002],[49,-0.0001],[55,-0.0009],[56,-0.0012],[60,-0.0037],[61,-0.0041],[62,-0.0042],[63,-0.0041],[65,-0.0031],[69,-0.0004],[70,0],[72,-0.0002],[74,0.0002],[82,0.0001],[99,0.0004],[100,0.0001],[101,0.0002],[102,-0.0001],[103,0.0002],[105,0.0002],[106,-0.0004],[107,0.0002],[108,-0.0061],[109,-0.0006],[110,-0.0031],[111,0.0007],[112,0.0017],[113,-0.0011],[114,-0.0013],[115,-0.0001],[135,0.0002],[136,-0.0006],[137,0.0002],[140,0.0004],[167,0.0002],[168,0.0001],[180,0.0002],[181,0],[182,0.0001],[183,-0.0002],[184,0.0002]]
       }
       pinky_01_r_ctrl.Location.Z {
        type=MovieSceneScriptingFloatChannel
        default=0
        thinned_from=185
        keys [[0,0],[56,0.0005],[61,0.0015],[64,0.0014],[68,0.0006],[70,0.0001],[72,0.0005],[74,0],[82,-0.0001],[99,0.0001],[100,0.0004],[101,0.0003],[102,0.0023],[103,0.0009],[104,-0.0002],[105,0],[106,-0.0004],[107,0],[108,-0.0009],[109,-0.0004],[110,-0.0013],[111,0.001],[112,0.0023],[113,0.0004],[114,0.0001],[115,0.0002],[134,-0.0001],[135,0],[136,-0.0001],[137,0],[138,-0.0003],[139,0],[169,0],[171,0.0005],[172,0],[180,0],[181,0.0005],[182,-0.0005],[183,-0.0005],[184,0]]
       }
       pinky_01_r_ctrl.Rotation.X {
        type=MovieSceneScriptingFloatChannel
        default=-9.8821
        thinned_from=185
        keys [[0,-9.4282],[1,-9.4433],[2,-9.4092],[3,-9.445],[5,-9.44],[7,-9.4849],[8,-9.4738],[9,-9.5258],[11,-9.5093],[12,-9.5351],[13,-9.4962],[14,-9.5534],[15,-9.485],[16,-9.4735],[17,-9.5242],[18,-9.4766],[20,-9.4919],[21,-9.4676],[22,-9.4945],[23,-9.4869],[24,-9.5106],[25,-9.4825],[26,-9.5295],[28,-9.4756],[29,-9.5415],[30,-9.5129],[32,-9.5085],[34,-9.557],[36,-9.5004],[37,-9.5401],[38,-9.4895],[39,-9.5239],[40,-9.4847],[42,-9.521],[45,-9.5023],[55,-9.7312],[56,-9.7907],[57,-9.8935],[60,-10.2707],[61,-10.3561],[62,-10.3896],[68,-10.3945],[69,-10.3411],[70,-10.2235],[71,-9.7309],[72,-9.2986],[74,-9.2886],[76,-9.189],[77,-9.0265],[78,-9.1098],[80,-9.0723],[82,-8.9897],[83,-8.7488],[84,-8.5298],[85,-8.5235],[86,-8.5492],[88,-8.4178],[89,-8.4262],[90,-8.3577],[91,-8.3668],[92,-8.3043],[93,-8.298],[94,-8.332],[95,-8.2608],[96,-8.4506],[97,-8.4246],[98,-8.5881],[99,-9.2204],[100,-9.7182],[101,-11.1548],[102,-11.8782],[103,-10.2709],[104,-12.855],[105,-12.9869],[106,-10.007],[107,-10.2416],[108,-10.0479],[109,-10.0017],[110,-10.1354],[111,-10.3677],[112,-10.3839],[113,-10.3624],[114,-10.4757],[115,-10.715],[118,-10.8809],[119,-10.8375],[120,-10.7485],[123,-10.668],[124,-10.5374],[125,-10.5048],[126,-10.5542],[127,-10.5574],[129,-10.4915],[130,-10.5149],[131,-10.4834],[132,-10.4978],[133,-10.6089],[134,-10.7667],[135,-10.8167],[136,-10.7846],[137,-10.7786],[138,-10.6543],[139,-10.5587],[140,-10.4989],[141,-10.5571],[142,-10.5052],[143,-10.4919],[144,-10.5253],[145,-10.5217],[146,-10.5418],[148,-10.5356],[149,-10.5654],[154,-10.5168],[156,-10.5131],[157,-10.5453],[160,-10.5745],[163,-10.5426],[165,-10.5563],[166,-10.6455],[167,-10.9937],[168,-11.5021],[169,-10.803],[170,-10.2846],[171,-9.8593],[172,-9.6471],[173,-9.6082],[174,-9.4862],[175,-9.4986],[176,-9.4214],[177,-9.4542],[178,-9.5194],[179,-9.5459],[180,-9.5167],[181,-9.6094],[182,-9.9573],[183,-10.1082],[184,-9.8821]]
       }
       pinky_01_r_ctrl.Rotation.Y {
        type=MovieSceneScriptingFloatChannel
        default=-7.1194
        thinned_from=185
        keys [[0,-6.5837],[1,-6.5615],[2,-6.4643],[3,-6.6393],[4,-6.5468],[5,-6.4888],[6,-6.5015],[8,-6.4198],[9,-6.4741],[10,-6.5419],[11,-6.4963],[12,-6.5479],[13,-6.4833],[14,-6.5334],[15,-6.5398],[16,-6.523],[17,-6.5457],[18,-6.6617],[19,-6.6361],[20,-6.7037],[22,-6.6746],[23,-6.5205],[24,-6.7026],[26,-6.6569],[27,-6.583],[28,-6.6689],[29,-6.599],[30,-6.5547],[31,-6.6309],[32,-6.5669],[33,-6.54],[34,-6.6357],[35,-6.532],[36,-6.6422],[37,-6.5628],[38,-6.7014],[39,-6.5833],[40,-6.5498],[41,-6.609],[42,-6.5692],[43,-6.6524],[44,-6.5632],[53,-6.3049],[55,-6.224],[56,-6.146],[57,-6.0192],[59,-5.7079],[60,-5.5676],[61,-5.4673],[62,-5.4289],[64,-5.4525],[67,-5.5131],[68,-5.5198],[69,-5.482],[70,-5.3989],[71,-5.1682],[72,-4.9698],[73,-4.9459],[74,-4.935],[75,-5.137],[76,-5.1413],[77,-5.1592],[78,-5.0745],[81,-5.0357],[82,-5.0321],[83,-5.323],[84,-5.1971],[85,-5.3303],[86,-5.3868],[87,-5.3802],[88,-5.2898],[90,-5.3323],[92,-5.2066],[93,-5.1937],[94,-5.1175],[95,-5.2222],[96,-5.3913],[97,-5.6909],[98,-6.1398],[99,-6.8677],[100,-7.0825],[101,-6.566],[102,-4.1725],[103,-3.3279],[104,-3.6369],[105,-3.8483],[106,-5.16],[107,-5.5682],[108,-5.4525],[109,-5.417],[110,-5.2793],[111,-4.646],[112,-9.9165],[113,-10.5419],[114,-10.8284],[115,-11.3721],[116,-11.6231],[117,-12.1008],[118,-12.3195],[119,-12.5121],[120,-12.7445],[121,-12.6733],[122,-12.9187],[123,-13.0336],[124,-13.58],[125,-13.4346],[126,-13.3487],[127,-13.381],[128,-13.6818],[129,-13.6633],[131,-13.7104],[132,-13.6811],[133,-13.4257],[134,-13.1139],[135,-13.2053],[136,-13.5087],[138,-13.9271],[139,-13.9896],[140,-14.1542],[141,-14.2757],[142,-14.2448],[144,-14.1213],[145,-13.9677],[147,-13.9947],[148,-13.9443],[149,-14.1128],[150,-14.0038],[151,-14.0019],[152,-14.034],[156,-13.9298],[157,-13.964],[158,-13.9492],[159,-14.0949],[161,-14.118],[162,-14.1952],[163,-14.1357],[164,-14.2541],[165,-14.1594],[166,-13.9926],[167,-12.6812],[168,-9.1793],[169,-6.8055],[170,-6.0487],[171,-6.0418],[172,-6.1486],[173,-6.3702],[174,-6.1945],[175,-6.6518],[176,-6.1632],[177,-6.4648],[178,-6.4091],[179,-6.2761],[180,-6.4567],[181,-6.4233],[182,-6.2764],[183,-6.4922],[184,-7.1194]]
       }
       pinky_01_r_ctrl.Rotation.Z {
        type=MovieSceneScriptingFloatChannel
        default=-49.0487
        thinned_from=185
        keys [[0,-48.6097],[1,-48.3696],[3,-48.4136],[4,-48.1304],[5,-48.4814],[7,-48.2661],[10,-48.2496],[11,-48.3302],[12,-48.126],[13,-48.3366],[14,-48.3929],[15,-48.5978],[16,-48.9079],[17,-48.7945],[18,-49.0518],[20,-48.9236],[21,-49.1685],[22,-49.1163],[24,-49.2577],[25,-49.405],[26,-49.0647],[28,-49.1906],[29,-48.9058],[31,-49.1906],[32,-49.0759],[33,-49.1203],[34,-48.8056],[35,-49.0933],[40,-49.3806],[42,-49.2648],[43,-49.508],[44,-49.5727],[45,-49.5055],[46,-49.3124],[47,-49.0059],[49,-48.1031],[51,-46.8981],[55,-43.9841],[56,-42.5137],[57,-39.9363],[58,-36.7158],[59,-33.3163],[60,-30.2019],[61,-27.8368],[62,-26.6849],[64,-25.9328],[65,-25.611],[66,-25.3542],[67,-25.1843],[68,-25.1228],[69,-25.353],[70,-25.8595],[71,-26.9197],[72,-27.714],[74,-27.7416],[75,-27.3254],[76,-26.774],[77,-25.8732],[78,-24.8328],[82,-24.9165],[83,-26.3402],[84,-27.0061],[85,-27.1267],[86,-26.699],[87,-26.9437],[88,-27.2597],[89,-27.2278],[90,-27.1157],[91,-27.1498],[92,-27.3609],[93,-27.7767],[94,-28.6706],[95,-28.4877],[96,-29.1675],[98,-30.645],[99,-30.2358],[100,-30.1829],[101,-15.9455],[102,-11.6707],[103,-14.1388],[104,-20.3223],[105,-20.3825],[106,-18.7317],[107,-19.6824],[108,-19.0425],[109,-20.1104],[110,-19.7855],[111,-20.6225],[112,-40.7027],[113,-47.0184],[114,-49.1182],[115,-50.2547],[116,-51.1445],[117,-51.5031],[118,-51.2141],[119,-51.4281],[120,-51.9491],[121,-51.9943],[124,-52.3461],[125,-51.998],[126,-52.0666],[127,-51.9421],[129,-52.1879],[130,-51.9199],[131,-51.9655],[132,-52.1271],[133,-51.9193],[134,-51.9021],[135,-52.2733],[136,-52.9947],[137,-53.4284],[138,-54.8632],[139,-56.0295],[141,-56.9475],[143,-57.371],[144,-57.3735],[146,-57.1572],[150,-57.3894],[152,-57.4498],[154,-57.9083],[155,-57.755],[160,-57.3345],[161,-57.384],[162,-57.5758],[163,-57.4464],[164,-57.7127],[165,-57.6074],[166,-57.1125],[167,-54.8118],[168,-49.3459],[169,-50.7626],[170,-51.6913],[171,-55.5783],[172,-57.4434],[173,-58.9829],[174,-59.2778],[175,-58.7148],[176,-58.9017],[177,-58.4125],[178,-58.0815],[179,-57.9261],[180,-57.4606],[181,-55.8747],[182,-44.9582],[183,-44.6778],[184,-49.0487]]
       }
       pinky_01_r_ctrl.Scale.X {type=MovieSceneScriptingFloatChannel, default=1, const [1,185]}
       pinky_01_r_ctrl.Scale.Y {type=MovieSceneScriptingFloatChannel, default=1, const [1,185]}
       pinky_01_r_ctrl.Scale.Z {type=MovieSceneScriptingFloatChannel, default=1, const [1,185]}
       pinky_02_r_ctrl.Location.X {
        type=MovieSceneScriptingFloatChannel
        default=-0.0001
        thinned_from=185
        keys [[0,-0.0001],[68,-0.0004],[70,-0.0001],[72,-0.0003],[74,-0.0001],[82,-0.0002],[97,-0.0001],[99,-0.0003],[100,-0.0035],[101,-0.003],[102,-0.0009],[103,-0.0004],[105,-0.0001],[106,-0.0006],[107,-0.0001],[108,-0.004],[109,-0.0008],[110,-0.0049],[111,-0.0036],[112,-0.008],[113,0],[135,-0.0001],[136,-0.0002],[137,-0.0001],[138,-0.0002],[139,-0.0001],[169,-0.0001],[171,-0.0002],[180,-0.0001],[181,-0.0005],[182,-0.0006],[184,-0.0001]]
       }
       pinky_02_r_ctrl.Location.Y {
        type=MovieSceneScriptingFloatChannel
        default=0.0001
        thinned_from=185
        keys [[0,0.0001],[99,0],[100,-0.0029],[101,-0.0017],[102,0],[105,0.0001],[106,-0.0004],[107,0.0001],[108,-0.0036],[109,-0.0006],[110,-0.003],[111,-0.0011],[112,-0.0041],[113,-0.0005],[114,-0.0007],[115,-0.0003],[118,0],[135,0.0001],[136,-0.0004],[137,0.0001],[138,-0.0004],[139,0],[140,-0.0001],[168,0],[169,-0.0002],[170,-0.0006],[171,-0.0007],[172,0],[180,0],[181,-0.0003],[183,-0.0003],[184,0.0001]]
       }
       pinky_02_r_ctrl.Location.Z {
        type=MovieSceneScriptingFloatChannel
        default=0
        thinned_from=185
        keys [[0,0],[68,0.0004],[70,0.0001],[100,0.0001],[101,-0.0001],[102,0.0003],[104,0],[106,0.0002],[107,0],[108,0.0014],[109,0.0001],[110,0.0007],[111,-0.0006],[112,-0.0026],[113,-0.0001],[114,-0.0002],[115,0.0001],[135,0],[136,0.0005],[137,0],[138,0.0003],[139,0],[170,-0.0001],[171,-0.0002],[183,0.0002],[184,0]]
       }
       pinky_02_r_ctrl.Rotation.X {
        type=MovieSceneScriptingFloatChannel
        default=-15.2844
        thinned_from=185
        keys [[0,-16.3407],[1,-16.3576],[2,-16.3373],[4,-16.3879],[5,-16.3352],[9,-16.312],[10,-16.3402],[11,-16.3101],[12,-16.3601],[13,-16.3119],[15,-16.3457],[16,-16.297],[22,-16.3209],[25,-16.247],[26,-16.3003],[28,-16.2455],[33,-16.1876],[34,-16.2248],[35,-16.1797],[36,-16.2247],[37,-16.1947],[38,-16.2264],[40,-16.1979],[45,-16.2019],[48,-16.0735],[55,-15.4205],[56,-15.2165],[57,-14.8637],[59,-13.9715],[60,-13.5622],[61,-13.2642],[62,-13.1423],[68,-13.0941],[70,-13.0072],[72,-13.2731],[74,-13.3242],[75,-13.4475],[76,-13.5231],[77,-13.6732],[78,-13.7614],[82,-13.7846],[83,-13.8248],[84,-13.9469],[90,-14.0973],[91,-14.1036],[93,-14.0663],[94,-13.9753],[95,-14.1171],[96,-14.0942],[97,-14.0427],[98,-13.9593],[99,-13.7778],[100,-12.2368],[101,-14.923],[102,-13.8094],[103,-13.427],[104,-12.2939],[105,-12.0436],[106,-13.3922],[107,-13.2643],[108,-13.3944],[109,-13.4479],[110,-13.4551],[111,-13.3439],[112,-10.9116],[113,-10.4766],[114,-10.2804],[115,-10.371],[116,-10.3432],[118,-10.2043],[119,-10.0323],[120,-9.9137],[123,-9.7408],[124,-9.8252],[125,-9.8487],[126,-9.7449],[127,-9.7434],[128,-9.8576],[131,-9.8472],[133,-9.8045],[134,-9.6888],[135,-9.7262],[136,-9.8442],[137,-9.8465],[138,-9.8793],[139,-9.9423],[140,-9.9049],[141,-9.9749],[142,-9.9546],[143,-9.9116],[144,-9.9103],[146,-9.9755],[148,-9.9386],[149,-10.0376],[151,-9.9665],[152,-10.0047],[154,-9.9759],[155,-9.9784],[156,-9.9453],[157,-9.9757],[158,-9.9853],[159,-10.0813],[160,-10.0968],[162,-10.084],[163,-10.1301],[165,-10.0812],[166,-10.1448],[167,-10.5471],[168,-11.2531],[169,-12.0491],[170,-13.2117],[171,-13.8385],[173,-14.2284],[174,-14.2351],[175,-14.1042],[176,-14.2662],[177,-14.2166],[178,-14.3998],[179,-14.537],[180,-14.5751],[181,-14.6582],[182,-14.8217],[183,-15.0743],[184,-15.2844]]
       }
       pinky_02_r_ctrl.Rotation.Y {
        type=MovieSceneScriptingFloatChannel
        default=-0.3859
        thinned_from=185
        keys [[0,-1.6534],[1,-1.6773],[2,-1.6487],[3,-1.7038],[4,-1.7203],[5,-1.6455],[6,-1.6519],[7,-1.6323],[8,-1.6257],[9,-1.6132],[10,-1.6526],[11,-1.6106],[12,-1.6808],[13,-1.613],[14,-1.6567],[15,-1.6604],[16,-1.5923],[17,-1.578],[18,-1.5951],[19,-1.6201],[20,-1.6115],[21,-1.6061],[22,-1.6256],[23,-1.5711],[24,-1.5599],[25,-1.5236],[26,-1.5969],[27,-1.5371],[28,-1.5215],[29,-1.5181],[30,-1.4823],[31,-1.4673],[32,-1.4781],[33,-1.4433],[34,-1.4934],[35,-1.4329],[36,-1.4933],[37,-1.4528],[38,-1.4956],[39,-1.4816],[40,-1.4572],[41,-1.4773],[42,-1.4769],[43,-1.4591],[44,-1.4753],[45,-1.4666],[46,-1.4416],[47,-1.402],[48,-1.3494],[49,-1.2853],[50,-1.2115],[51,-1.1295],[52,-1.041],[53,-0.9476],[55,-0.7527],[56,-0.5668],[57,-0.2455],[58,0.1518],[59,0.5658],[60,0.937],[61,1.2062],[62,1.314],[67,1.3371],[68,1.3434],[69,1.3713],[70,1.3956],[71,1.3217],[72,1.2358],[74,1.2034],[75,1.1245],[76,1.0748],[77,0.9734],[78,0.9123],[82,0.896],[83,0.8676],[84,0.7792],[85,0.7501],[86,0.7406],[87,0.715],[88,0.6965],[90,0.6667],[91,0.6618],[92,0.6791],[93,0.6901],[94,0.7582],[95,0.6517],[96,0.669],[97,0.708],[98,0.7694],[99,0.9],[100,1.8098],[101,-0.0291],[102,0.8788],[103,1.1378],[104,1.7771],[105,1.8941],[106,1.162],[107,1.2409],[108,1.1569],[109,1.1242],[110,1.1193],[111,1.177],[112,2.2909],[113,2.4598],[114,2.5111],[115,2.4904],[116,2.4978],[118,2.5333],[119,2.574],[120,2.6011],[123,2.6377],[124,2.6201],[125,2.6152],[126,2.6366],[127,2.637],[128,2.6133],[131,2.6155],[133,2.6245],[134,2.6476],[135,2.6404],[136,2.6161],[137,2.6157],[139,2.595],[140,2.6035],[141,2.5877],[143,2.6017],[144,2.602],[146,2.5875],[148,2.5958],[149,2.5734],[151,2.5896],[152,2.5809],[155,2.5869],[156,2.5943],[158,2.5854],[159,2.5633],[160,2.5596],[162,2.5626],[163,2.5516],[165,2.5632],[166,2.5456],[167,2.4424],[168,2.2135],[169,1.8833],[170,1.2648],[171,0.8515],[172,0.7146],[173,0.5644],[174,0.5599],[175,0.6613],[176,0.5353],[177,0.5743],[178,0.4277],[179,0.313],[180,0.2801],[181,0.2089],[182,0.0625],[183,-0.1749],[184,-0.3859]]
       }
       pinky_02_r_ctrl.Rotation.Z {
        type=MovieSceneScriptingFloatChannel
        default=-49.9502
        thinned_from=185
        keys [[0,-58.7654],[1,-58.9216],[2,-58.7344],[3,-59.0937],[4,-59.2023],[5,-58.7153],[6,-58.7559],[9,-58.5021],[10,-58.7604],[11,-58.4849],[12,-58.9448],[13,-58.5007],[14,-58.7866],[15,-58.811],[16,-58.3645],[17,-58.271],[19,-58.5476],[21,-58.4552],[22,-58.5835],[23,-58.2256],[24,-58.1516],[25,-57.9118],[26,-58.3949],[27,-58.0015],[29,-57.8759],[30,-57.6387],[31,-57.5384],[32,-57.6106],[33,-57.3792],[34,-57.7119],[35,-57.3099],[36,-57.7109],[37,-57.4426],[38,-57.7264],[40,-57.4716],[41,-57.6048],[42,-57.6025],[43,-57.4841],[44,-57.5919],[45,-57.5226],[46,-57.3234],[48,-56.5872],[50,-55.4872],[55,-51.8279],[56,-50.3427],[57,-47.7747],[58,-44.5976],[59,-41.2851],[60,-38.3107],[61,-36.1481],[62,-35.2711],[68,-34.9651],[70,-34.4159],[71,-35.1855],[72,-36.0776],[74,-36.4044],[75,-37.1881],[76,-37.6745],[77,-38.6428],[78,-39.222],[82,-39.3744],[83,-39.6384],[84,-40.4463],[85,-40.7083],[86,-40.7934],[90,-41.4522],[91,-41.4948],[93,-41.2444],[94,-40.6354],[95,-41.5867],[96,-41.4312],[97,-41.0845],[98,-40.5246],[99,-39.3239],[100,-29.8384],[101,-47.3291],[102,-39.5026],[103,-37.0453],[104,-30.0729],[105,-28.5841],[106,-36.8346],[107,-36.0251],[108,-36.8547],[109,-37.1924],[110,-37.239],[111,-36.3886],[112,-21.6954],[113,-19.6097],[114,-18.5183],[115,-19.0158],[116,-18.8604],[117,-18.4991],[118,-18.0887],[119,-17.1364],[120,-16.4809],[121,-16.0897],[123,-15.5289],[124,-15.9931],[125,-16.1226],[126,-15.5509],[127,-15.5428],[128,-16.171],[131,-16.114],[132,-15.9499],[133,-15.8789],[134,-15.2424],[135,-15.4483],[136,-16.0974],[137,-16.1099],[138,-16.2919],[139,-16.6382],[140,-16.4316],[141,-16.8184],[142,-16.7062],[143,-16.4691],[144,-16.4617],[145,-16.6881],[146,-16.8215],[147,-16.6883],[148,-16.6179],[149,-17.1646],[150,-16.9379],[151,-16.7718],[152,-16.9827],[153,-16.8807],[154,-16.8237],[155,-16.8374],[156,-16.6549],[157,-16.8224],[158,-16.8754],[159,-17.4061],[160,-17.492],[162,-17.4214],[163,-17.6768],[164,-17.5725],[165,-17.4059],[166,-17.7612],[167,-19.9988],[168,-23.9928],[169,-28.631],[170,-35.7125],[171,-39.7451],[173,-42.3433],[174,-42.386],[175,-41.499],[176,-42.5987],[177,-42.2595],[178,-43.517],[179,-44.475],[180,-44.7453],[181,-45.3324],[182,-46.5039],[183,-48.3584],[184,-49.9502]]
       }
       pinky_02_r_ctrl.Scale.X {type=MovieSceneScriptingFloatChannel, default=1, const [1,185]}
       pinky_02_r_ctrl.Scale.Y {type=MovieSceneScriptingFloatChannel, default=1, const [1,185]}
       pinky_02_r_ctrl.Scale.Z {type=MovieSceneScriptingFloatChannel, default=1, const [1,185]}
       pinky_03_r_ctrl.Location.X {
        type=MovieSceneScriptingFloatChannel
        default=0.0002
        thinned_from=185
        keys [[0,0.0002],[55,-0.0004],[61,-0.002],[63,-0.002],[70,0.0001],[72,-0.0001],[74,0.0002],[82,0.0001],[101,0.0001],[102,-0.0005],[103,-0.0001],[105,0.0002],[106,-0.0003],[107,0.0002],[108,-0.0032],[109,-0.0004],[110,-0.0037],[111,-0.0003],[112,-0.0011],[113,-0.0001],[114,-0.0003],[115,0.0001],[135,0.0002],[136,-0.0003],[137,0.0002],[138,0.0001],[169,0.0001],[170,-0.0003],[171,-0.0004],[172,0.0002],[180,0.0002],[182,-0.0002],[184,0.0002]]
       }
       pinky_03_r_ctrl.Location.Y {
        type=MovieSceneScriptingFloatChannel
        default=-0.0002
        thinned_from=185
        keys [[0,-0.0002],[101,-0.0003],[102,-0.0009],[103,-0.0005],[105,-0.0002],[107,-0.0002],[108,-0.0008],[109,-0.0004],[111,-0.0011],[112,-0.0025],[113,0],[114,0.0001],[135,-0.0002],[136,0],[137,-0.0002],[180,-0.0002],[182,-0.0003],[183,-0.0001],[184,-0.0002]]
       }
       pinky_03_r_ctrl.Location.Z {
        type=MovieSceneScriptingFloatChannel
        default=0
        thinned_from=185
        keys [[0,0],[101,0],[102,-0.0002],[104,0],[105,0],[106,0.0003],[107,0],[108,0.0017],[109,0.0002],[110,0.0013],[111,0.0004],[112,0.0007],[113,-0.0004],[114,-0.0005],[115,-0.0001],[135,0],[136,0.0003],[137,0],[138,0.0002],[139,0],[168,-0.0001],[171,-0.0006],[172,-0.0001],[180,0],[181,-0.0002],[182,0.0002],[184,0]]
       }
       pinky_03_r_ctrl.Rotation.X {
        type=MovieSceneScriptingFloatChannel
        default=-7.79
        thinned_from=185
        keys [[0,-9.0319],[1,-9.0571],[2,-9.0473],[4,-9.0872],[5,-9.03],[6,-9.0048],[7,-8.9997],[8,-9.0173],[9,-9.0003],[10,-9.0444],[11,-9.0126],[12,-9.0443],[13,-9.0022],[15,-9.0352],[17,-8.9294],[18,-8.9895],[19,-9.0068],[22,-8.9887],[23,-8.9271],[24,-8.9384],[25,-8.9252],[26,-8.9487],[27,-8.9269],[28,-8.9301],[30,-8.8808],[32,-8.8752],[33,-8.8569],[34,-8.8916],[35,-8.8557],[36,-8.9001],[37,-8.8622],[38,-8.9208],[39,-8.8817],[40,-8.87],[41,-8.8986],[45,-8.8882],[47,-8.8448],[50,-8.7169],[55,-8.4089],[56,-8.2841],[57,-8.0686],[59,-7.5249],[60,-7.2762],[61,-7.0962],[62,-7.0247],[68,-7.0084],[70,-6.9074],[71,-7.014],[72,-7.1497],[74,-7.2274],[75,-7.4376],[76,-7.5661],[77,-7.8025],[78,-7.8816],[82,-7.8751],[83,-7.885],[84,-8],[85,-8.0933],[86,-8.1587],[87,-8.2018],[88,-8.2166],[89,-8.2611],[90,-8.2811],[92,-8.2073],[93,-8.1866],[94,-8.0565],[95,-8.2324],[96,-8.2548],[97,-8.1795],[98,-8.0571],[99,-7.8118],[100,-6.3301],[101,-8.8048],[102,-8.7268],[103,-8.7061],[104,-7.5605],[105,-7.4168],[106,-8.4932],[107,-8.1605],[108,-8.1908],[109,-8.0734],[110,-8.0236],[111,-7.9213],[112,-4.6262],[113,-4.281],[114,-4.0674],[115,-3.7587],[116,-3.6365],[117,-3.5059],[118,-3.4243],[119,-3.3001],[120,-3.1936],[121,-3.1591],[122,-3.1054],[123,-3.0357],[124,-3.0064],[125,-3.0269],[126,-2.9915],[127,-2.9887],[128,-3.0126],[129,-3.0176],[131,-2.9948],[132,-2.99],[133,-3.0077],[134,-2.9546],[135,-2.942],[136,-2.969],[137,-2.8603],[138,-2.8234],[139,-2.8457],[140,-2.797],[141,-2.8164],[142,-2.8202],[143,-2.7827],[144,-2.801],[145,-2.8266],[146,-2.8347],[147,-2.8207],[148,-2.8174],[149,-2.8456],[150,-2.8441],[151,-2.8288],[152,-2.8544],[153,-2.8408],[155,-2.8581],[157,-2.8564],[159,-2.8799],[160,-2.8986],[161,-2.8812],[162,-2.8767],[163,-2.8806],[164,-2.8735],[165,-2.8593],[166,-2.9159],[167,-3.3058],[168,-4.2037],[169,-5.1344],[170,-6.0295],[171,-6.5315],[172,-6.7254],[173,-6.8393],[174,-6.8871],[175,-6.7459],[176,-6.9435],[177,-6.8848],[179,-7.1424],[180,-7.1587],[181,-7.2425],[182,-7.4579],[183,-7.69],[184,-7.79]]
       }
       pinky_03_r_ctrl.Rotation.Y {
        type=MovieSceneScriptingFloatChannel
        default=2.4651
        thinned_from=185
        keys [[0,2.3964],[4,2.3897],[6,2.3995],[9,2.4],[10,2.3949],[11,2.3986],[12,2.395],[13,2.3999],[15,2.396],[17,2.4079],[19,2.3993],[22,2.4014],[23,2.4082],[28,2.4078],[33,2.4154],[34,2.4119],[35,2.4156],[36,2.411],[37,2.4149],[38,2.4088],[40,2.4141],[68,2.43],[70,2.4201],[72,2.4395],[74,2.4453],[75,2.4558],[84,2.4643],[93,2.4599],[94,2.4633],[96,2.4572],[99,2.4634],[100,2.3668],[101,2.4345],[102,2.4273],[103,2.4292],[104,2.4613],[106,2.447],[107,2.4606],[110,2.4642],[111,2.419],[112,1.8132],[113,1.8049],[114,1.7203],[115,1.5888],[116,1.5333],[117,1.472],[118,1.4323],[119,1.3699],[120,1.315],[121,1.2968],[122,1.2682],[123,1.2302],[124,1.2143],[125,1.2256],[126,1.206],[127,1.2044],[128,1.2176],[129,1.2204],[131,1.2078],[132,1.2051],[133,1.2149],[134,1.1855],[135,1.1784],[136,1.194],[137,1.132],[138,1.1104],[139,1.1236],[140,1.0957],[141,1.1066],[142,1.1088],[143,1.087],[144,1.0977],[145,1.1126],[146,1.1173],[147,1.1092],[148,1.1072],[149,1.1235],[150,1.1226],[151,1.1139],[152,1.1286],[153,1.1208],[155,1.1307],[156,1.1289],[157,1.1297],[159,1.1432],[160,1.1539],[161,1.1439],[162,1.1414],[163,1.1436],[164,1.1396],[165,1.1314],[166,1.1623],[167,1.373],[168,1.7758],[169,2.0844],[170,2.2957],[171,2.3762],[172,2.4008],[173,2.4132],[174,2.418],[175,2.4032],[176,2.4233],[177,2.4177],[179,2.4395],[181,2.4464],[183,2.4637],[184,2.4651]]
       }
       pinky_03_r_ctrl.Rotation.Z {
        type=MovieSceneScriptingFloatChannel
        default=-31.2264
        thinned_from=185
        keys [[0,-38.3371],[1,-38.4819],[2,-38.4258],[4,-38.6555],[5,-38.3257],[6,-38.1815],[7,-38.1518],[8,-38.2529],[9,-38.1555],[10,-38.4089],[11,-38.2262],[12,-38.4082],[13,-38.1661],[15,-38.3564],[17,-37.7477],[18,-38.0933],[19,-38.1926],[22,-38.0889],[23,-37.7344],[24,-37.7993],[25,-37.7239],[26,-37.8592],[27,-37.7336],[28,-37.7517],[30,-37.4689],[32,-37.4364],[33,-37.3313],[34,-37.5306],[35,-37.3247],[36,-37.5797],[37,-37.3618],[38,-37.6987],[39,-37.4737],[40,-37.4067],[41,-37.5711],[43,-37.5032],[45,-37.5115],[47,-37.2636],[49,-36.8157],[53,-35.5199],[55,-34.772],[56,-34.0588],[57,-32.8269],[58,-31.3042],[59,-29.7182],[60,-28.2966],[61,-27.2671],[62,-26.8573],[68,-26.7609],[70,-26.1752],[71,-26.7837],[72,-27.5601],[74,-28.0078],[75,-29.2102],[76,-29.9457],[77,-31.2975],[78,-31.7487],[82,-31.7114],[83,-31.7682],[84,-32.4253],[85,-32.9589],[86,-33.3325],[87,-33.5792],[88,-33.6635],[89,-33.918],[90,-34.0324],[92,-33.6105],[93,-33.4925],[94,-32.7484],[95,-33.7534],[96,-33.8826],[97,-33.4522],[98,-32.7546],[99,-31.3532],[100,-22.8547],[101,-37.0264],[102,-36.5847],[103,-36.4663],[104,-29.9128],[105,-29.0917],[106,-35.2258],[107,-33.3431],[108,-33.4987],[109,-32.8443],[110,-32.5558],[111,-31.9875],[112,-12.7399],[113,-10.7401],[114,-9.4351],[115,-7.5093],[116,-6.7423],[117,-5.9176],[118,-5.3984],[119,-4.6081],[120,-3.9224],[121,-3.7008],[122,-3.3536],[123,-2.9016],[124,-2.7115],[125,-2.845],[126,-2.6147],[127,-2.596],[128,-2.7515],[129,-2.7841],[130,-2.7195],[131,-2.6361],[132,-2.6048],[133,-2.7199],[134,-2.3737],[135,-2.2917],[136,-2.4559],[137,-1.7567],[138,-1.5154],[139,-1.6603],[140,-1.3389],[141,-1.4681],[142,-1.4933],[143,-1.2461],[144,-1.3668],[145,-1.5355],[146,-1.5888],[147,-1.4966],[148,-1.4744],[149,-1.6602],[150,-1.6501],[151,-1.5499],[152,-1.7181],[153,-1.6286],[154,-1.678],[155,-1.742],[156,-1.721],[157,-1.7308],[158,-1.8014],[159,-1.8855],[160,-2.0076],[161,-1.8937],[162,-1.864],[163,-1.8897],[164,-1.8435],[165,-1.7501],[166,-2.1307],[167,-4.6414],[169,-15.8671],[170,-21.139],[171,-24.0298],[172,-25.1302],[173,-25.7856],[174,-26.0573],[175,-25.2467],[176,-26.3811],[177,-26.0444],[179,-27.5211],[180,-27.6148],[181,-28.0932],[182,-29.3279],[183,-30.6527],[184,-31.2264]]
       }
       pinky_03_r_ctrl.Scale.X {type=MovieSceneScriptingFloatChannel, default=1, const [1,185]}
       pinky_03_r_ctrl.Scale.Y {type=MovieSceneScriptingFloatChannel, default=1, const [1,185]}
       pinky_03_r_ctrl.Scale.Z {type=MovieSceneScriptingFloatChannel, default=1, const [1,185]}
       hand_r_ik_ctrl.Location.X {
        type=MovieSceneScriptingFloatChannel
        default=-28.5693
        thinned_from=185
        keys [[0,9.8392],[2,10.1439],[4,10.3531],[5,10.4188],[6,10.4409],[8,10.4338],[9,10.3954],[10,10.2682],[12,9.9141],[13,9.7912],[14,9.6331],[15,9.5788],[16,9.5605],[17,9.575],[19,9.6845],[20,9.6949],[21,9.814],[23,9.9438],[24,9.9929],[25,10.1065],[27,10.2435],[28,10.3554],[29,10.4423],[30,10.4881],[31,10.4567],[33,10.3516],[34,10.3382],[35,10.3509],[37,10.2233],[38,10.1811],[42,10.1271],[43,10.0766],[44,9.9522],[45,10.147],[46,10.5497],[47,11.1301],[48,11.8361],[49,13.023],[50,14.3131],[51,16.0931],[52,16.3597],[53,14.1797],[54,13.0304],[55,9.3756],[56,7.5158],[57,5.0684],[58,2.2185],[59,-1.8793],[60,-6.2256],[61,-10.3757],[62,-13.0597],[63,-15.532],[64,-17.5391],[65,-19.1561],[67,-22.5832],[69,-26.2197],[70,-27.9105],[71,-29.45],[72,-30.9168],[73,-32.1342],[74,-33.1327],[75,-33.9752],[76,-34.6589],[77,-35.2358],[78,-35.5757],[79,-35.7842],[80,-36.112],[82,-36.486],[84,-36.6322],[93,-37.0804],[96,-36.9571],[99,-36.5691],[102,-36.5862],[103,-36.52],[104,-36.0829],[105,-36.0037],[106,-36.192],[108,-36.281],[109,-36.5387],[110,-37.3156],[111,-38.6835],[112,-40.5908],[113,-42.7805],[114,-44.8398],[115,-47.848],[116,-49.3029],[117,-50.5775],[118,-51.3962],[119,-51.9373],[120,-52.1943],[125,-52.477],[131,-52.4252],[133,-52.5287],[136,-52.9967],[137,-52.894],[139,-51.7554],[141,-50.8435],[142,-50.3052],[144,-49.694],[146,-49.3338],[154,-48.8894],[161,-48.9121],[163,-48.6012],[165,-47.8614],[166,-47.0918],[167,-45.165],[168,-38.8616],[169,-30.6501],[170,-21.1989],[171,-11.8777],[172,-7.369],[173,-3.4379],[174,-0.5759],[175,2.2979],[176,4.8195],[177,6.5473],[178,7.9851],[179,8.6901],[180,8.0499],[181,4.2371],[182,-14.2252],[183,-19.8456],[184,-28.5693]]
       }
       hand_r_ik_ctrl.Location.Y {
        type=MovieSceneScriptingFloatChannel
        default=-41.6067
        thinned_from=185
        keys [[0,10.4111],[1,10.4009],[2,10.3193],[3,10.2037],[4,10.0395],[6,9.5833],[7,9.3805],[9,9.04],[10,8.8189],[11,8.6258],[12,8.4975],[13,8.4345],[14,8.421],[15,8.4792],[16,8.586],[17,8.7364],[19,9.1637],[20,9.2824],[21,9.5187],[22,9.6677],[23,9.7693],[24,9.826],[25,9.8324],[26,9.7724],[27,9.6093],[28,9.3578],[29,9.1658],[30,8.9459],[31,8.6922],[32,8.4895],[33,8.3595],[34,8.3409],[37,8.4117],[40,8.6957],[41,8.8194],[43,9.1387],[44,9.2491],[45,9.5658],[46,10.0824],[47,10.8424],[48,11.9426],[49,13.5625],[50,15.5584],[51,17.9364],[52,19.5424],[53,19.9744],[54,20.717],[55,19.9315],[56,20.9617],[57,23.2893],[58,26.2621],[59,28.9896],[60,31.3526],[61,33.0085],[62,34.3017],[63,34.8173],[64,35.2312],[65,35.8215],[66,36.1707],[68,37.0231],[70,37.5004],[72,38.042],[75,38.0567],[76,38.0022],[77,38.042],[82,37.7625],[85,37.4738],[94,37.4518],[96,37.2537],[97,36.958],[98,36.3951],[99,35.5439],[100,33.5193],[101,31.7128],[102,29.3082],[103,26.9703],[104,23.0643],[105,21.2455],[106,17.8842],[107,16.0279],[108,12.2913],[109,9.4756],[110,7.5385],[111,6.2898],[112,5.7201],[113,5.8308],[114,6.5157],[115,8.1453],[116,9.1712],[117,10.2561],[118,11.1445],[119,12],[120,12.6181],[121,12.9232],[122,13.3691],[123,13.763],[124,14.0803],[125,14.201],[127,14.2639],[128,14.3799],[130,14.4704],[131,14.477],[133,14.3197],[134,13.9231],[135,13.1495],[136,11.474],[137,9.3852],[138,6.7984],[139,4.8697],[140,3.6267],[141,2.8151],[142,2.4493],[143,2.4521],[144,2.5011],[145,2.5707],[146,2.5946],[147,2.6313],[149,2.647],[150,2.6718],[151,2.7107],[152,2.7651],[153,2.8704],[154,2.9694],[155,3.0598],[156,3.1259],[157,3.1855],[158,3.1919],[159,3.2091],[160,3.2012],[162,3.2134],[163,3.2333],[164,3.2394],[165,3.2307],[166,3.2394],[167,3.532],[168,5.3234],[169,6.7306],[170,6.9981],[171,6.9039],[172,5.9532],[173,6.9261],[174,7.4047],[175,7.0088],[176,6.6521],[177,7.5631],[178,9.2026],[179,11.9126],[180,15.1883],[181,16.6552],[182,-1.2271],[183,-28.8337],[184,-41.6067]]
       }
       hand_r_ik_ctrl.Location.Z {
        type=MovieSceneScriptingFloatChannel
        default=182.0742
        thinned_from=185
        keys [[0,66.4205],[9,65.7746],[13,65.8989],[22,66.3899],[26,66.1052],[31,65.1472],[35,65.0499],[44,66.2715],[45,66.1149],[46,65.5577],[47,64.6783],[48,63.5522],[49,61.8957],[50,59.8277],[51,56.8649],[52,54.3974],[53,53.4928],[54,51.4757],[55,51.2248],[56,48.741],[57,45.275],[58,40.7497],[59,36.6151],[60,33.0233],[61,30.7221],[62,29.0765],[63,28.909],[64,28.8871],[65,28.7887],[67,29.2197],[68,29.6162],[69,30.1245],[70,30.5463],[73,31.248],[76,31.7097],[77,31.9352],[79,32.0976],[82,32.0602],[83,31.7801],[84,31.5964],[85,31.4881],[96,31.0651],[98,31.1036],[99,31.2993],[100,31.9969],[101,32.557],[103,34.2229],[104,35.7182],[105,36.3617],[106,37.338],[107,37.8006],[108,38.6212],[109,39.1402],[110,39.357],[111,39.2736],[112,39.0068],[114,37.6665],[115,36.3435],[117,34.6768],[118,33.9968],[119,33.3915],[120,32.9962],[124,32.1515],[125,32.0695],[133,32.3778],[134,32.6316],[135,33.1874],[136,34.5179],[137,36.0766],[138,37.7782],[139,38.7826],[140,39.2395],[141,39.4941],[144,39.583],[150,39.9902],[159,40.1066],[164,40.4415],[165,40.455],[166,40.1563],[167,39.0214],[168,34.8991],[169,29.1493],[170,21.4677],[171,11.0787],[172,2.6313],[173,-6.1133],[174,-11.2203],[175,-12.801],[176,-12.0892],[177,-10.0061],[178,-6.8558],[179,-2.811],[180,2.4239],[181,9.4513],[182,34.0419],[183,135.9755],[184,182.0742]]
       }
       hand_r_ik_ctrl.Rotation.X {
        type=MovieSceneScriptingFloatChannel
        default=-50.5424
        thinned_from=185
        keys [[0,-35.8251],[1,-34.8567],[2,-34.3206],[3,-33.5645],[4,-32.2219],[5,-31.281],[6,-30.6002],[7,-30.1167],[8,-28.9811],[9,-27.9791],[10,-27.685],[11,-27.712],[12,-27.9248],[13,-28.0016],[14,-28.9065],[15,-30.0512],[16,-30.9212],[17,-31.6768],[18,-33.0419],[19,-33.8215],[20,-33.9621],[21,-34.8016],[22,-35.4585],[23,-35.8002],[24,-35.9905],[25,-35.9375],[26,-35.354],[27,-34.5295],[29,-32.304],[30,-31.3956],[31,-31.2419],[32,-31.2369],[33,-30.9681],[34,-30.5785],[35,-30.5269],[36,-30.9421],[37,-31.1659],[38,-31.6804],[39,-32.1302],[42,-33.0395],[43,-33.7726],[44,-34.5845],[45,-34.545],[46,-33.7942],[47,-32.2421],[48,-29.2362],[49,-22.7035],[50,-11.8731],[51,4.4116],[52,19.5677],[53,31.0215],[54,42.2308],[55,49.7689],[56,58.7977],[57,66.2493],[58,69.8506],[59,71.6293],[60,72.4465],[61,73.6676],[62,74.2556],[63,76.6968],[64,79.4684],[65,82.0237],[66,85.7543],[67,89.8162],[68,94.6183],[71,107.9152],[72,111.9145],[73,114.2429],[74,115.7728],[76,119.4196],[77,121.6172],[81,122.1982],[82,122.0141],[83,121.1442],[85,120.4244],[92,120.9776],[94,121.6294],[95,121.2479],[96,121.3556],[97,120.7176],[98,119.3073],[99,116.7061],[100,108.3538],[101,107.6237],[102,109.5574],[103,105.0938],[104,113.4296],[105,115.5164],[106,114.4989],[107,115.7271],[108,112.4142],[109,106.9331],[110,81.0391],[111,-57.2004],[112,-81.664],[113,-102.8995],[114,-110.9696],[115,-117.4245],[116,-120.9052],[117,-124.0549],[119,-127.2735],[120,-128.1855],[122,-128.676],[123,-129.1917],[126,-129.892],[127,-129.8346],[128,-130.3301],[131,-130.2975],[133,-129.6485],[134,-127.2282],[135,-124.4899],[136,-120.4529],[137,-115.2668],[138,-108.4295],[139,-102.7756],[140,-99.0724],[141,-97.2399],[142,-96.5222],[144,-96.6507],[145,-96.5141],[150,-97.098],[154,-97.864],[157,-97.7624],[163,-98.1026],[165,-98.5909],[167,-98.0221],[168,-97.461],[169,-87.092],[170,-62.8406],[171,-24.5073],[172,-1.3351],[173,10.1006],[174,12.0075],[175,11.1195],[176,10.3116],[177,8.2897],[178,5.3675],[179,1.2398],[180,-6.6232],[181,-17.9827],[182,-35.4481],[183,-46.5546],[184,-50.5424]]
       }
       hand_r_ik_ctrl.Rotation.Y {
        type=MovieSceneScriptingFloatChannel
        default=53.7149
        thinned_from=185
        keys [[0,57.179],[1,57.5883],[3,57.7693],[5,58.3068],[6,58.1318],[7,58.3061],[9,59.045],[10,59.1686],[12,58.9984],[13,59.0186],[14,58.9162],[15,58.4986],[16,58.2395],[18,57.4803],[19,57.2735],[21,57.2052],[22,56.8779],[24,57.0121],[26,56.973],[27,57.2833],[29,57.4716],[30,57.8063],[32,57.5064],[34,57.5235],[35,57.7897],[39,58.0244],[41,57.8889],[43,58.1827],[44,58.058],[45,58.4441],[46,59.5141],[47,61.0974],[48,63.2899],[49,65.9718],[50,68.4688],[51,69.5738],[53,69.2283],[54,67.0035],[55,65.9332],[56,60.7348],[57,52.1406],[58,41.6111],[59,32.3186],[60,25.3673],[61,21.7979],[62,20.7669],[63,21.8234],[64,22.2754],[66,20.8383],[67,19.3781],[68,17.702],[69,16.1157],[70,14.1914],[71,11.4132],[72,8.6743],[73,6.8865],[74,5.3649],[75,4.0713],[76,3.3031],[77,2.7268],[78,2.2954],[79,1.8798],[80,1.5155],[81,1.1403],[82,0.9823],[83,1.1111],[84,2.0141],[85,2.4666],[86,2.252],[87,2.6821],[88,2.9722],[89,2.9024],[90,2.876],[91,2.8946],[92,2.8535],[93,3.0058],[94,3.009],[95,3.4852],[96,3.9786],[97,5.1406],[98,7.1281],[99,8.3926],[100,11.4935],[101,15.7667],[102,17.292],[103,26.2037],[104,26.9092],[105,28.0886],[106,41.1726],[107,45.7668],[108,59.016],[109,71.7051],[110,83.7248],[111,80.3932],[112,72.0819],[113,59.6607],[114,49.1788],[115,34.5935],[116,28.2392],[117,23.7337],[118,21.256],[119,19.0057],[120,17.4574],[121,16.9195],[123,15.5008],[124,15.2238],[125,15.616],[127,15.5546],[128,15.4141],[129,15.1406],[130,15.1329],[131,15.0893],[133,15.1673],[134,15.1179],[135,15.6791],[136,17.7014],[137,20.6694],[138,22.377],[139,23.4387],[140,24.8403],[141,25.2969],[142,25.6264],[143,26.1259],[144,26.5187],[147,26.8719],[148,26.8628],[149,26.6677],[150,26.5778],[151,26.7263],[153,26.6208],[154,26.698],[155,26.867],[157,26.921],[158,27.0684],[159,27.1039],[161,27.4787],[162,27.8527],[163,28.2905],[165,29.0025],[166,29.5035],[167,30.9083],[168,35.4865],[169,37.997],[170,45.1141],[171,45.8116],[172,36.1241],[173,24.1476],[174,14.4085],[175,10.8436],[176,12.9143],[177,16.2559],[179,28.4469],[180,33.7339],[181,38.188],[182,44.1647],[183,58.5688],[184,53.7149]]
       }
       hand_r_ik_ctrl.Rotation.Z {
        type=MovieSceneScriptingFloatChannel
        default=-134.3373
        thinned_from=185
        keys [[0,-90.8405],[1,-89.8621],[2,-89.3976],[4,-88.1023],[6,-87.7312],[7,-87.8588],[9,-87.0425],[11,-87.387],[13,-88.3227],[15,-89.7789],[16,-90.2124],[17,-90.3686],[18,-91.0409],[19,-91.4597],[20,-91.3736],[21,-91.7587],[23,-92.0068],[25,-91.9394],[26,-91.4086],[27,-91.1094],[29,-89.9304],[30,-89.6665],[31,-90.0222],[32,-90.1833],[33,-90.0258],[34,-89.6523],[35,-89.8942],[36,-90.3822],[37,-90.4549],[39,-91.2017],[42,-91.5327],[44,-92.7509],[45,-92.0988],[46,-90.4602],[47,-87.6357],[48,-83.1192],[49,-74.9731],[50,-62.4152],[51,-44.3649],[52,-27.5695],[53,-14.646],[54,-1.9001],[55,6.6148],[56,18.5042],[57,31.4372],[58,42.4762],[59,52.2614],[60,60.8504],[61,67.8863],[62,72.6586],[65,82.501],[68,91.529],[69,94.2172],[71,98.862],[72,99.8943],[74,98.3919],[75,97.4101],[77,94.4654],[78,92.616],[80,90.3814],[81,89.4352],[82,88.8344],[83,88.4909],[84,88.8311],[85,88.8968],[86,88.5355],[89,88.9592],[93,88.6914],[95,89.0038],[96,89.6285],[97,90.0055],[98,91.0545],[99,92.8639],[100,93.5563],[101,96.2171],[102,95.4917],[103,97.33],[104,102.496],[105,101.7603],[106,89.7962],[107,85.9769],[108,75.1251],[109,61.3949],[110,25.6639],[111,-118.7515],[112,-133.3335],[113,-152.0512],[115,-171.9381],[117,-179.1316],[118,-182.2014],[119,-184.573],[120,-185.7676],[121,-186.1919],[123,-187.5885],[125,-186.9932],[127,-187.1483],[129,-186.4595],[131,-186.591],[133,-187.3563],[134,-188.312],[135,-187.1192],[136,-182.4577],[137,-175.7809],[138,-168.4504],[139,-163.8831],[140,-161.4217],[141,-159.4533],[142,-158.2409],[146,-156.9605],[165,-159.03],[166,-158.6154],[167,-157.271],[168,-152.9246],[169,-143.5607],[170,-125.8173],[171,-92.7394],[172,-72.7153],[173,-60.7943],[174,-55.0091],[175,-52.8442],[176,-52.7071],[177,-53.694],[178,-55.5891],[179,-57.9286],[180,-63.2044],[181,-71.5825],[182,-95.1321],[183,-119.7926],[184,-134.3373]]
       }
       hand_r_ik_ctrl.Scale.X {type=MovieSceneScriptingFloatChannel, default=1, const [1,185]}
       hand_r_ik_ctrl.Scale.Y {type=MovieSceneScriptingFloatChannel, default=1, const [1,185]}
       hand_r_ik_ctrl.Scale.Z {type=MovieSceneScriptingFloatChannel, default=1, const [1,185]}
       Softness_2 {type=MovieSceneScriptingFloatChannel, default=0, const [0,185]}
       hand_r_ikorient_ctrl.Rotation.Roll {
        type=MovieSceneScriptingFloatChannel
        default=0
        thinned_from=185
        keys [[0,0],[172,0],[173,29.6691],[174,0],[175,28.1688],[176,0],[177,0],[178,28.5681],[179,0],[180,30.0636],[181,0],[184,0]]
       }
       hand_r_ikorient_ctrl.Rotation.Pitch {
        type=MovieSceneScriptingFloatChannel
        default=0
        thinned_from=185
        keys [[0,0],[172,0],[173,10.7647],[174,0],[175,-16.8861],[176,0],[177,0],[178,-15.4637],[179,0],[180,-18.3882],[181,0],[184,0]]
       }
       hand_r_ikorient_ctrl.Rotation.Yaw {
        type=MovieSceneScriptingFloatChannel
        default=0
        thinned_from=185
        keys [[0,0],[172,0],[173,27.1908],[174,0],[175,26.6331],[176,0],[177,0],[178,18.6621],[179,0],[180,22.6546],[181,0],[184,0]]
       }
       arm_r_pv_ik_ctrl.Location.X {
        type=MovieSceneScriptingFloatChannel
        default=21.8458
        thinned_from=185
        keys [[0,-0.4577],[1,-0.375],[2,-0.2873],[3,-0.1977],[4,0.0292],[5,0.2744],[6,0.4959],[7,0.7069],[8,0.9427],[9,1.12],[10,1.31],[11,1.4333],[12,1.5404],[13,1.612],[14,1.5534],[15,1.4169],[16,1.2452],[17,1.0544],[18,0.8193],[19,0.6537],[20,0.5491],[21,0.3571],[22,0.2244],[23,0.1302],[24,0.0918],[25,0.1235],[26,0.2167],[27,0.4173],[28,0.6914],[29,0.9093],[30,1.1789],[31,1.4019],[32,1.5485],[33,1.6565],[35,1.7357],[36,1.722],[37,1.6896],[38,1.5924],[39,1.5081],[40,1.4098],[41,1.2893],[42,1.1621],[43,1.053],[44,0.9103],[45,0.8145],[46,0.867],[47,0.9537],[48,1.0107],[49,1.036],[50,1.0076],[51,1.0332],[52,1.3078],[53,1.7288],[54,2.2348],[55,3.019],[56,3.3939],[57,3.0943],[58,2.1441],[59,0.4761],[60,-1.7844],[61,-4.1084],[62,-5.8784],[63,-6.7177],[64,-7.0613],[65,-7.0717],[66,-6.8363],[67,-6.53],[68,-6.3699],[69,-6.3098],[70,-6.2864],[71,-6.3744],[72,-6.439],[73,-6.3692],[74,-6.3194],[75,-6.213],[76,-6.0788],[77,-6.0487],[79,-5.7791],[81,-5.5584],[82,-5.4614],[83,-5.2437],[84,-5.1367],[85,-5.1107],[86,-5.2338],[87,-5.3177],[88,-5.3514],[92,-5.3523],[93,-5.3099],[94,-5.2342],[95,-5.0512],[96,-4.7746],[97,-4.3816],[98,-3.7739],[99,-3.0512],[100,-1.877],[101,-1.1004],[102,-0.607],[103,-0.5269],[105,-1.3565],[106,-2.32],[107,-2.8306],[108,-3.7413],[109,-4.2439],[110,-4.4876],[111,-4.6409],[112,-5.0275],[113,-5.2802],[114,-5.2714],[115,-5.0387],[116,-4.886],[117,-4.7155],[118,-4.5349],[119,-4.3436],[120,-4.216],[121,-4.1725],[122,-4.1606],[123,-4.1266],[124,-4.1446],[125,-4.2089],[126,-4.2528],[127,-4.2475],[129,-4.3284],[130,-4.3385],[131,-4.2983],[132,-4.1767],[133,-4.0843],[134,-3.8233],[135,-3.5379],[136,-3.2768],[137,-3.185],[138,-3.2259],[139,-3.1491],[140,-3.0081],[141,-3.0071],[142,-3.1925],[143,-3.371],[144,-3.4876],[145,-3.619],[146,-3.6655],[147,-3.6898],[149,-3.6784],[151,-3.6971],[152,-3.6853],[153,-3.7274],[155,-3.7792],[156,-3.7349],[157,-3.7171],[158,-3.6655],[159,-3.6233],[160,-3.5105],[161,-3.4567],[163,-3.1266],[164,-2.9724],[165,-2.9099],[166,-2.9277],[167,-2.9251],[168,-3.4982],[169,-5.7897],[170,-10.0527],[171,-15.7983],[172,-20.2884],[173,-25.0792],[174,-26.82],[175,-26.0356],[176,-24.3842],[177,-23.8659],[178,-23.5763],[179,-23.6648],[180,-24.5051],[181,-24.376],[182,-5.4092],[183,21.1599],[184,21.8458]]
       }
       arm_r_pv_ik_ctrl.Location.Y {
        type=MovieSceneScriptingFloatChannel
        default=165.0033
        thinned_from=185
        keys [[0,76.8727],[13,75.7018],[23,76.6876],[26,76.5158],[31,75.2794],[35,75.0791],[44,76.5135],[45,76.4393],[46,75.9296],[47,75.0673],[48,73.9111],[49,72.2903],[50,70.2636],[52,64.8454],[53,62.7809],[54,59.9871],[55,57.8324],[56,54.2739],[57,49.6455],[58,43.7141],[59,37.3866],[60,31.1807],[61,26.3883],[62,23.2808],[63,22.6345],[64,22.8503],[65,23.6784],[66,25.2892],[67,27.4408],[69,32.8189],[70,35.4217],[71,37.7915],[72,39.7163],[73,40.9399],[74,41.9044],[75,42.6424],[76,43.0535],[77,43.7169],[78,43.9642],[79,43.9733],[82,43.4838],[85,41.7376],[86,41.2714],[87,41.0898],[95,41.138],[96,41.0595],[97,40.6849],[98,39.9184],[99,38.8043],[100,36.2033],[101,34.0405],[102,31.403],[103,29.1024],[104,26.4126],[105,25.5827],[106,24.4778],[107,24.176],[108,24.1666],[109,24.6359],[110,25.1919],[111,25.6531],[112,26.4119],[114,27.7518],[115,28.916],[117,30.2936],[118,30.7311],[119,30.9736],[120,31.0865],[125,31.1581],[127,31.0892],[134,31.3823],[135,31.3548],[137,31.1793],[139,31.1536],[142,30.9069],[144,30.8845],[147,31.1744],[151,31.4208],[157,31.1848],[161,31.147],[165,30.8627],[166,30.2196],[167,27.4393],[168,18.8849],[169,8.4917],[170,-1.2805],[171,-8.2414],[172,-10.6591],[173,-11.4298],[174,-10.6706],[175,-9.4039],[176,-8.3041],[177,-7.2897],[178,-6.0806],[179,-4.7396],[180,-2.8988],[181,-0.1622],[182,19.0056],[183,117.8507],[184,165.0033]]
       }
       arm_r_pv_ik_ctrl.Location.Z {
        type=MovieSceneScriptingFloatChannel
        default=92.4416
        thinned_from=185
        keys [[0,-6.7032],[2,-6.8796],[3,-6.9123],[5,-6.6316],[6,-6.5199],[7,-6.3144],[8,-6.0288],[9,-5.7781],[10,-5.4735],[11,-5.244],[12,-4.9707],[13,-4.7338],[14,-4.6399],[15,-4.751],[16,-4.8884],[17,-5.0789],[18,-5.3645],[19,-5.569],[20,-5.6789],[21,-5.9856],[22,-6.2145],[23,-6.4023],[24,-6.5413],[25,-6.6481],[26,-6.7202],[27,-6.6618],[29,-6.4776],[31,-6.1625],[32,-6.0445],[33,-5.8939],[34,-5.8388],[35,-5.7128],[36,-5.6427],[37,-5.643],[38,-5.6669],[39,-5.6786],[40,-5.7022],[41,-5.7373],[42,-5.7515],[44,-5.7432],[45,-6.1511],[46,-6.8712],[47,-7.9107],[48,-9.1084],[49,-10.6961],[50,-12.3154],[51,-14.3751],[52,-14.8686],[53,-12.8692],[54,-11.8562],[55,-8.1761],[56,-6.9246],[57,-6.4693],[58,-6.8451],[59,-6.8965],[60,-6.9765],[61,-6.8705],[62,-7.0001],[63,-6.8051],[64,-6.7891],[65,-6.7554],[66,-6.6249],[67,-6.291],[68,-5.7973],[69,-4.986],[70,-3.8081],[71,-2.4181],[72,-0.6185],[73,1.2656],[74,2.9263],[75,4.5236],[76,6.036],[77,7.4831],[78,8.5506],[79,9.2613],[80,10.1052],[81,10.7897],[82,11.2909],[83,11.563],[84,11.5051],[85,11.2575],[86,10.9115],[87,10.7467],[88,10.7019],[89,10.6959],[91,10.8184],[92,10.8548],[93,11.0346],[94,11.2751],[95,11.6941],[96,12.2351],[97,12.6292],[98,12.9418],[99,13.0648],[100,12.7958],[101,12.3023],[102,11.1724],[103,9.9401],[104,8.023],[105,7.1513],[106,5.999],[107,5.403],[108,4.6697],[109,4.4954],[110,4.5558],[111,4.596],[112,4.4352],[113,4.3019],[114,4.4573],[115,5.1149],[116,5.6269],[117,6.2092],[118,6.7271],[119,7.1921],[120,7.4866],[122,7.7335],[123,7.8309],[124,7.9056],[125,7.884],[126,7.8436],[127,7.8557],[128,7.8852],[129,7.8845],[130,7.9166],[131,7.9835],[132,8.1413],[133,8.2536],[134,8.6178],[135,9.0123],[136,9.5284],[137,9.9797],[138,10.3743],[139,10.701],[140,10.8779],[141,10.8222],[142,10.401],[143,10.0504],[144,9.8142],[145,9.6392],[146,9.5316],[149,9.455],[150,9.4162],[153,9.2136],[154,9.1282],[155,9.1402],[156,9.1884],[159,9.3804],[160,9.4661],[161,9.4965],[162,9.5532],[163,9.5371],[164,9.41],[165,9.1104],[166,8.3681],[167,6.8554],[168,2.8396],[169,-0.0133],[170,-1.1438],[171,-1.6051],[172,-3.2666],[173,-6.5088],[174,-10.3799],[175,-13.8392],[176,-16.1459],[177,-17.6551],[178,-19.0492],[179,-19.8954],[180,-20.9522],[181,-21.5555],[182,-9.1353],[183,55.2926],[184,92.4416]]
       }
       arm_r_pv_ik_ctrl.Rotation.X {
        type=MovieSceneScriptingFloatChannel
        default=-85.7678
        thinned_from=185
        keys [[0,-59.6654],[3,-59.2056],[5,-58.2804],[6,-58.0323],[7,-57.6033],[8,-56.9268],[9,-56.4633],[10,-56.2207],[11,-56.3036],[12,-56.2306],[13,-56.0283],[14,-56.3781],[18,-58.296],[23,-59.3915],[24,-59.5006],[26,-59.2804],[27,-58.8862],[30,-57.1182],[31,-56.8959],[32,-56.8524],[34,-56.5533],[35,-56.1839],[36,-56.2391],[38,-56.7631],[39,-56.9052],[41,-57.3698],[42,-57.49],[43,-57.7456],[44,-58.2049],[45,-58.0483],[46,-57.0803],[47,-55.5037],[48,-53.0419],[49,-49.0043],[50,-43.9997],[51,-38.0216],[52,-33.1804],[53,-30.2561],[54,-26.3784],[55,-24.3031],[56,-20.3401],[57,-16.4333],[58,-12.6324],[59,-9.7756],[60,-7.5178],[61,-6.3764],[62,-5.5704],[63,-6.6512],[64,-7.9333],[65,-8.8175],[66,-10.4216],[67,-11.9115],[69,-15.5339],[70,-17.0261],[71,-18.0842],[72,-18.7405],[73,-19.0153],[74,-19.2249],[75,-19.2689],[76,-19.1702],[79,-18.6417],[80,-18.4076],[81,-18.0635],[82,-17.7852],[83,-17.3053],[84,-17.2216],[85,-17.4753],[86,-18.1266],[87,-18.4566],[88,-18.6038],[89,-18.8024],[90,-18.9377],[92,-19.0653],[93,-18.9812],[94,-18.767],[95,-18.3081],[96,-17.6992],[97,-17.1803],[98,-16.7807],[99,-16.9811],[100,-19.102],[101,-21.4192],[102,-25.1357],[103,-28.7386],[104,-34.365],[105,-37.1052],[106,-42.2238],[107,-44.6982],[108,-49.5637],[109,-53.3546],[110,-56.4522],[112,-61.6445],[113,-63.8355],[114,-65.7182],[115,-68.5968],[117,-71.6459],[118,-72.707],[119,-73.3901],[120,-73.6966],[124,-73.7579],[129,-73.4055],[131,-73.4738],[133,-73.8399],[137,-76.373],[139,-77.2401],[140,-77.501],[141,-77.4448],[143,-76.4113],[145,-75.7063],[156,-74.7674],[161,-74.6388],[163,-74.0612],[165,-72.8994],[166,-71.2789],[167,-66.599],[168,-52.4141],[169,-35.2129],[170,-17.5641],[171,-2.7488],[172,0.5591],[173,2.3235],[174,0.493],[175,-1.5995],[176,-2.2835],[177,-1.9502],[178,-1.3503],[179,-0.3867],[180,-0.9727],[181,-5.368],[182,-35.2102],[183,-63.079],[184,-85.7678]]
       }
       arm_r_pv_ik_ctrl.Rotation.Y {
        type=MovieSceneScriptingFloatChannel
        default=70.5271
        thinned_from=185
        keys [[0,48.0851],[3,48.2506],[5,49.0324],[6,49.2252],[7,49.5724],[8,50.0699],[10,50.8425],[11,51.0372],[13,51.6601],[14,51.6776],[17,50.8159],[18,50.3448],[20,49.8866],[22,49.1586],[24,48.751],[26,48.7238],[29,49.4163],[30,49.8011],[31,50.0445],[34,50.4269],[35,50.6786],[37,50.65],[41,50.4516],[43,50.511],[44,50.3958],[45,49.9953],[46,49.4433],[47,48.5877],[48,47.4893],[49,45.9364],[50,43.9084],[51,41.1904],[52,38.929],[53,37.8185],[54,35.8875],[55,35.5329],[56,32.6208],[57,27.709],[58,21.0191],[59,14.2154],[60,7.6144],[61,2.614],[62,-0.9835],[63,-1.3441],[64,-0.8714],[65,-0.1567],[66,1.9471],[67,4.5507],[68,7.7289],[69,11.1431],[70,14.6818],[71,18.0308],[72,21.5709],[73,24.7443],[74,27.4764],[75,30.0048],[76,32.2572],[77,34.3317],[78,35.928],[79,36.9564],[80,38.083],[81,38.9203],[82,39.5343],[83,39.7499],[84,39.6046],[86,39.064],[87,38.9512],[88,38.982],[89,39.1012],[92,39.6528],[94,40.2936],[95,40.8339],[96,41.4956],[97,41.8814],[98,42.1186],[99,42.2038],[100,41.957],[101,41.5353],[102,40.3075],[103,38.7933],[104,36.0807],[105,34.6629],[106,32.1649],[107,30.6738],[108,27.9998],[109,26.2058],[110,24.8153],[111,23.3354],[112,21.217],[113,19.2978],[115,17.0905],[116,16.6615],[117,16.3562],[118,16.3742],[120,16.9754],[121,17.1317],[123,17.6392],[124,17.7765],[125,17.7251],[126,17.7714],[127,17.8825],[128,17.9551],[130,17.989],[132,18.1965],[133,18.2329],[135,18.3944],[137,18.6636],[139,18.8664],[140,18.936],[141,18.8638],[142,18.5241],[143,18.2956],[146,18.0772],[148,18.0778],[151,17.9914],[154,18.0511],[155,18.1494],[156,18.3274],[157,18.4556],[159,18.6205],[160,18.7998],[161,18.8898],[163,19.4239],[164,19.6348],[165,19.6003],[166,19.3749],[167,19.8582],[168,22.4578],[169,25.6536],[170,27.2708],[171,26.1699],[172,24.6728],[173,22.0256],[174,20.3041],[176,17.7964],[177,15.6648],[178,12.7095],[179,9.352],[180,5.8723],[181,5.2033],[182,20.1238],[183,51.5294],[184,70.5271]]
       }
       arm_r_pv_ik_ctrl.Rotation.Z {
        type=MovieSceneScriptingFloatChannel
        default=-12.8691
        thinned_from=185
        keys [[0,17.5967],[1,17.7063],[2,17.6981],[3,17.6303],[4,17.8534],[5,17.9537],[6,17.8749],[7,17.988],[8,18.3082],[9,18.5001],[10,18.548],[11,18.3632],[12,18.3176],[13,18.4448],[14,18.1019],[15,17.7076],[16,17.4265],[18,17.0594],[19,17.079],[20,17.0504],[21,17.068],[24,16.9456],[25,17.0594],[26,17.052],[27,17.1894],[28,17.4148],[30,17.7313],[31,17.6906],[32,17.602],[34,17.7599],[35,18.0856],[36,18.096],[38,17.8843],[39,17.8896],[40,17.8361],[41,17.837],[42,17.898],[43,17.9024],[44,17.7392],[45,18.0843],[46,19.0305],[47,20.4114],[48,22.5308],[49,25.8762],[50,29.9141],[51,34.3788],[52,37.729],[53,39.6614],[54,41.7234],[55,42.3125],[56,43.8019],[57,44.8709],[58,44.8912],[59,43.6779],[60,41.7614],[61,39.9446],[62,39.1426],[63,39.5598],[64,40.6411],[65,42.5564],[66,44.669],[68,49.7025],[69,51.7806],[71,55.6552],[72,57.1297],[73,57.8503],[74,58.3535],[75,58.7187],[76,58.8387],[77,59.3481],[78,59.4926],[79,59.4089],[80,59.2044],[82,59.0609],[83,58.772],[84,58.2062],[86,56.5612],[87,56.251],[90,55.9781],[93,56.1715],[96,56.8055],[97,56.4509],[98,55.296],[99,53.0248],[100,46.6043],[101,40.7076],[102,32.7312],[103,25.3573],[104,14.9711],[105,10.5272],[106,3.2253],[107,-0.2148],[108,-6.114],[109,-10.0501],[110,-13.1396],[111,-16.0862],[112,-18.5979],[113,-20.8719],[114,-22.4401],[115,-24.2495],[116,-24.9906],[117,-25.5761],[118,-25.9486],[119,-26.1593],[120,-26.2236],[121,-26.228],[124,-26.0393],[127,-25.9395],[129,-25.721],[131,-25.6781],[133,-26.003],[136,-27.9473],[137,-28.2843],[138,-28.3635],[139,-28.6328],[140,-28.9898],[141,-29.0737],[142,-28.9553],[144,-28.4491],[145,-28.125],[147,-27.7286],[150,-27.5154],[153,-27.4951],[160,-27.9107],[162,-27.9518],[163,-27.8876],[165,-27.5338],[166,-27.5161],[167,-28.0759],[168,-28.5384],[169,-27.7991],[170,-26.4353],[171,-26.8306],[172,-32.0717],[173,-38.6519],[174,-43.6898],[175,-45.5294],[176,-44.6989],[177,-42.6716],[178,-39.6349],[179,-35.274],[180,-30.9949],[181,-27.8953],[182,-35.977],[183,-12.7342],[184,-12.8691]]
       }
       arm_r_pv_ik_ctrl.Scale.X {type=MovieSceneScriptingFloatChannel, default=1, const [1,185]}
       arm_r_pv_ik_ctrl.Scale.Y {type=MovieSceneScriptingFloatChannel, default=1, const [1,185]}
       arm_r_pv_ik_ctrl.Scale.Z {type=MovieSceneScriptingFloatChannel, default=1, const [1,185]}
       body_ctrl.Location.X {
        type=MovieSceneScriptingFloatChannel
        default=2.2697
        thinned_from=182
        keys [[0,19.2116],[6,19.4191],[32,19.5038],[39,19.2192],[44,18.8932],[45,18.8682],[47,18.7242],[49,18.5104],[50,18.3717],[51,18.2814],[52,18.1268],[53,17.8532],[54,17.7023],[55,17.4242],[57,17.0515],[58,16.8292],[59,16.5596],[60,16.3477],[61,16.197],[62,16.1481],[63,16.0536],[64,16.0031],[67,15.9293],[70,16.0141],[71,16.0761],[72,16.1853],[74,16.3028],[79,16.3417],[85,16.2067],[88,16.1286],[99,16.3254],[107,16.6082],[113,17.0024],[135,17.3523],[139,17.5598],[141,17.6095],[152,17.6114],[160,17.1905],[164,17.0252],[166,16.9836],[167,16.9266],[168,16.7703],[169,16.5303],[170,16.1794],[171,15.7713],[172,15.5292],[174,14.8506],[176,13.9634],[177,13.8923],[178,13.9779],[179,14.4375],[180,15.7266],[181,17.7155],[182,16.813],[184,2.2697]]
       }
       body_ctrl.Location.Y {
        type=MovieSceneScriptingFloatChannel
        default=95.056
        thinned_from=182
        keys [[0,-10.5952],[6,-10.3368],[12,-10.1591],[14,-10.1519],[19,-10.2785],[26,-10.4045],[27,-10.4376],[29,-10.5454],[32,-10.6297],[35,-10.6347],[45,-10.5143],[46,-10.4822],[47,-10.4102],[48,-10.2333],[49,-9.984],[50,-9.6433],[51,-9.4022],[52,-8.901],[53,-7.8727],[54,-7.1264],[55,-5.7387],[56,-4.8339],[57,-3.7879],[58,-2.8532],[59,-1.8838],[60,-1.0848],[61,-0.3109],[62,0.0016],[63,0.5909],[64,1.02],[65,1.1862],[66,1.4261],[67,1.518],[68,1.5659],[69,1.571],[70,1.5384],[71,1.4728],[72,1.3857],[73,1.3329],[74,1.3119],[75,1.2981],[76,1.2766],[77,1.2618],[78,1.2525],[79,1.2465],[81,1.2029],[83,1.1632],[84,1.1244],[85,1.0767],[86,1.0073],[87,0.9492],[88,0.8952],[89,0.8261],[90,0.7664],[91,0.7051],[92,0.6421],[93,0.6231],[94,0.5952],[95,0.5726],[96,0.5479],[97,0.531],[98,0.5181],[99,0.5104],[100,0.5226],[101,0.5412],[102,0.5738],[103,0.6295],[106,0.7116],[108,0.7193],[109,0.7173],[110,0.7104],[111,0.7169],[112,0.732],[113,0.7664],[114,0.8104],[115,0.8942],[116,0.951],[117,1.0153],[118,1.0682],[119,1.1138],[120,1.1429],[121,1.1578],[122,1.1792],[124,1.2067],[125,1.2174],[127,1.2314],[129,1.2498],[130,1.2566],[131,1.2575],[133,1.2363],[134,1.2016],[135,1.1739],[136,1.1391],[137,1.1016],[138,1.057],[139,1.0182],[140,0.9831],[141,0.9549],[142,0.932],[143,0.9198],[144,0.9141],[146,0.9266],[147,0.9301],[151,0.9288],[152,0.9315],[153,0.9517],[155,1.0434],[156,1.1019],[157,1.1732],[158,1.2549],[159,1.322],[160,1.3756],[161,1.3906],[162,1.4089],[163,1.4113],[164,1.408],[165,1.3793],[166,1.3031],[167,1.074],[168,0.3745],[169,-0.566],[170,-1.4802],[171,-2.1373],[172,-2.4155],[173,-2.6586],[174,-2.7736],[175,-2.8188],[176,-2.7388],[177,-2.5626],[178,-2.4249],[179,-2.3607],[180,-2.2824],[181,-2.1042],[182,6.8014],[184,95.056]]
       }
       body_ctrl.Location.Z {
        type=MovieSceneScriptingFloatChannel
        default=52.8612
        thinned_from=182
        keys [[0,-9.7273],[4,-9.859],[5,-9.8679],[6,-9.8556],[9,-9.7531],[10,-9.7361],[12,-9.7618],[14,-9.8326],[20,-9.862],[26,-9.9569],[27,-9.9367],[29,-9.8096],[33,-9.7],[36,-9.7003],[39,-9.788],[42,-9.9171],[44,-10.0738],[45,-10.1174],[47,-10.3356],[48,-10.406],[49,-10.3893],[51,-10.0702],[52,-9.6627],[53,-8.677],[54,-7.9425],[55,-6.5313],[56,-5.6733],[57,-4.7604],[58,-3.9767],[59,-3.1849],[60,-2.5795],[61,-2.0957],[62,-1.9282],[63,-1.678],[64,-1.5771],[65,-1.5707],[66,-1.6258],[67,-1.7467],[68,-1.8949],[69,-2.0073],[70,-2.0774],[71,-2.0928],[72,-2.0468],[73,-1.9784],[75,-1.9285],[76,-1.8943],[77,-1.8744],[79,-1.848],[81,-1.8001],[82,-1.792],[83,-1.7906],[85,-1.8072],[86,-1.8238],[87,-1.833],[88,-1.8294],[89,-1.8318],[90,-1.8512],[91,-1.8831],[92,-1.9234],[93,-1.937],[94,-1.9568],[95,-1.9668],[96,-1.9699],[97,-1.9588],[98,-1.9313],[99,-1.912],[100,-1.9016],[101,-1.9052],[102,-1.9148],[103,-1.9298],[106,-1.9126],[107,-1.9184],[108,-1.9318],[112,-1.9448],[115,-1.9684],[116,-1.9654],[118,-1.9358],[119,-1.927],[121,-1.925],[125,-1.937],[127,-1.9637],[128,-1.9859],[129,-1.9895],[133,-1.9791],[134,-1.9712],[135,-1.9524],[136,-1.9253],[137,-1.9073],[138,-1.9016],[142,-1.9137],[143,-1.912],[144,-1.9059],[146,-1.873],[147,-1.8644],[149,-1.8596],[150,-1.869],[151,-1.892],[152,-1.9196],[154,-1.9822],[155,-1.9944],[156,-2.0016],[157,-1.9923],[159,-1.9446],[160,-1.9159],[161,-1.9036],[162,-1.88],[164,-1.8391],[165,-1.8294],[166,-1.8414],[167,-1.9314],[168,-2.6143],[169,-4.1799],[170,-6.4477],[171,-8.7927],[172,-10.3452],[173,-12.3683],[174,-14.0388],[175,-15.483],[176,-16.4282],[177,-16.9019],[178,-17.1176],[179,-17.0707],[180,-16.7815],[181,-16.0189],[182,-10.1545],[184,52.8612]]
       }
       body_ctrl.Rotation.X {
        type=MovieSceneScriptingFloatChannel
        default=60.5079
        thinned_from=182
        keys [[0,116.5706],[45,116.6254],[47,115.7159],[49,113.5827],[51,110.6239],[52,107.3897],[53,101.6165],[54,96.4612],[55,88.8468],[56,82.6104],[57,75.7834],[58,69.3447],[59,62.6604],[60,56.7694],[61,51.3457],[62,48.0921],[63,43.9286],[64,40.2095],[65,37.6064],[66,34.4646],[68,29.1481],[69,26.7633],[70,24.4543],[71,22.3965],[72,20.1062],[73,17.9212],[74,16.0416],[75,14.3186],[76,12.6682],[77,11.2302],[78,10.0759],[79,9.0664],[80,7.9593],[81,7.146],[82,6.5554],[83,6.1004],[84,5.6719],[86,4.7683],[87,4.4366],[88,4.1229],[89,3.88],[90,3.6594],[91,3.4679],[92,3.306],[93,3.0978],[95,2.7839],[96,2.6606],[97,2.558],[98,2.4797],[99,2.4376],[100,2.4619],[101,2.4779],[102,2.4888],[103,2.5762],[106,2.0706],[107,2],[109,1.7195],[110,1.6348],[111,1.6189],[113,1.6965],[114,1.7018],[115,1.6944],[116,1.6559],[118,1.5418],[119,1.5062],[120,1.5047],[121,1.499],[122,1.509],[124,1.5091],[125,1.5264],[126,1.5512],[127,1.553],[128,1.5332],[129,1.5253],[130,1.5216],[131,1.545],[132,1.6145],[133,1.668],[134,1.817],[135,1.92],[136,1.9261],[137,1.8606],[138,1.7666],[139,1.7429],[140,1.7948],[141,1.8743],[142,1.9905],[143,2.0848],[144,2.159],[145,2.2179],[146,2.2906],[147,2.3884],[148,2.4277],[149,2.5221],[150,2.589],[151,2.6412],[152,2.6687],[153,2.6556],[156,2.4044],[157,2.3031],[158,2.2187],[159,2.1855],[160,2.2188],[161,2.264],[162,2.4452],[163,2.6984],[164,3.0547],[165,3.4812],[166,4.0624],[167,5.3245],[168,8.8491],[169,12.6401],[170,16.214],[171,18.9337],[172,19.6717],[173,20.5725],[174,20.8965],[175,21.5544],[176,22.283],[177,22.5535],[179,23.2421],[180,23.6705],[181,23.7945],[182,31.5663],[184,60.5079]]
       }
       body_ctrl.Rotation.Y {
        type=MovieSceneScriptingFloatChannel
        default=3.8394
        thinned_from=182
        keys [[0,-8.1899],[1,-8.1136],[3,-8.0254],[4,-8.0786],[5,-8.2102],[7,-8.5607],[8,-8.7059],[9,-8.8152],[11,-9.2371],[12,-9.4638],[13,-9.559],[14,-9.5518],[15,-9.4513],[17,-9.1578],[18,-8.9833],[19,-8.8608],[20,-8.7877],[21,-8.57],[25,-8.0624],[26,-7.9994],[27,-8.1063],[28,-8.2515],[29,-8.2856],[30,-8.347],[31,-8.5131],[33,-8.9323],[34,-8.9808],[35,-8.9961],[36,-9.0481],[37,-9.0664],[39,-8.9808],[40,-8.9872],[41,-8.9207],[42,-8.7922],[43,-8.7286],[45,-8.7757],[46,-8.8844],[47,-8.9114],[48,-8.8382],[49,-8.6895],[50,-8.4813],[51,-8.3509],[52,-8.1118],[53,-7.4717],[54,-6.9164],[55,-6.0435],[56,-5.4876],[57,-4.8245],[58,-4.2763],[59,-3.7008],[60,-3.169],[61,-2.5709],[62,-2.2923],[63,-1.7284],[64,-1.1716],[65,-0.8908],[66,-0.3411],[67,0.0679],[68,0.6172],[69,1.0371],[70,1.2787],[71,1.3454],[72,1.1744],[73,0.8441],[74,0.6153],[75,0.3662],[76,0.0373],[77,-0.2408],[79,-0.5413],[80,-0.7372],[81,-0.989],[82,-1.1594],[83,-1.3043],[84,-1.396],[85,-1.422],[86,-1.3807],[87,-1.3287],[88,-1.301],[89,-1.254],[90,-1.1934],[91,-1.1361],[92,-1.0473],[93,-1.0145],[94,-0.9786],[95,-0.9695],[96,-0.9921],[97,-1.0759],[98,-1.2002],[99,-1.314],[100,-1.4352],[101,-1.5264],[102,-1.625],[103,-1.7883],[106,-1.983],[107,-2.018],[108,-2.1168],[110,-2.3286],[111,-2.4042],[112,-2.4341],[113,-2.4376],[114,-2.4509],[115,-2.4754],[116,-2.5172],[117,-2.5389],[118,-2.5813],[119,-2.6441],[120,-2.715],[121,-2.7569],[123,-2.9435],[124,-3],[126,-3.0464],[127,-3.0587],[128,-3.1062],[130,-3.1707],[132,-3.196],[134,-3.1958],[135,-3.1417],[136,-3.0688],[137,-3.0376],[138,-3.0686],[139,-3.1103],[140,-3.1228],[141,-3.119],[142,-3.1499],[144,-3.1727],[145,-3.151],[146,-3.0987],[147,-3.0593],[148,-3.0469],[149,-3.0501],[150,-3.0655],[151,-3.1155],[152,-3.1948],[153,-3.3098],[154,-3.5059],[155,-3.6628],[156,-3.8298],[158,-4.0294],[159,-4.0797],[160,-4.151],[161,-4.1813],[162,-4.2404],[163,-4.232],[165,-4.0498],[166,-3.965],[167,-3.9882],[168,-4.3964],[169,-5.0209],[170,-5.6343],[171,-6.1807],[172,-6.1724],[173,-5.9168],[174,-5.6198],[175,-5.0971],[176,-4.4942],[177,-4.045],[178,-3.6765],[179,-3.483],[180,-3.2495],[181,-2.8939],[182,0.3499],[184,3.8394]]
       }
       body_ctrl.Rotation.Z {
        type=MovieSceneScriptingFloatChannel
        default=10.253
        thinned_from=182
        keys [[0,10.3746],[1,10.4323],[3,10.6126],[4,10.7434],[6,11.0783],[9,11.4679],[11,11.8603],[12,12.0004],[13,12.0303],[14,11.9355],[17,11.4791],[18,11.2825],[19,11.1359],[20,11.0697],[21,10.916],[22,10.8437],[24,10.8452],[25,10.8741],[26,10.9408],[27,11.0773],[28,11.2577],[30,11.4396],[31,11.623],[32,11.7785],[33,11.8828],[34,11.8914],[37,11.7478],[38,11.6272],[40,11.3161],[41,11.1407],[42,10.9264],[43,10.7543],[44,10.708],[45,10.691],[46,10.7206],[47,10.772],[48,10.8531],[49,10.9783],[50,11.1372],[51,11.4056],[52,12.0735],[53,13.0679],[54,14.143],[55,15.5792],[56,16.8619],[57,18.0881],[58,19.2447],[59,20.1688],[60,20.8536],[61,21.3169],[62,21.6627],[63,22.1783],[64,22.8811],[65,23.3094],[66,23.8331],[67,23.9146],[68,23.6455],[69,23.5509],[71,23.4394],[72,23.6823],[74,25.0471],[76,26.4687],[77,26.9439],[79,27.3647],[80,27.674],[81,27.8995],[82,28.0667],[83,28.2912],[85,28.8958],[86,29.1365],[89,29.321],[95,29.4043],[97,29.3045],[99,28.9584],[100,28.6687],[101,28.4994],[106,27.9415],[108,27.8899],[110,27.9637],[112,27.904],[113,27.9163],[114,27.9979],[115,28.1979],[118,28.6019],[119,28.6748],[123,28.6798],[125,28.7672],[131,28.6777],[133,28.5472],[134,28.422],[143,28.1164],[145,28.0931],[146,28.0184],[147,27.8753],[148,27.8035],[151,27.3648],[152,27.2946],[153,27.3174],[159,27.9638],[160,28.0296],[161,28.0228],[164,27.7916],[166,27.4892],[167,27.2327],[168,26.2056],[169,25.8453],[170,25.3403],[171,25.6445],[173,26.5041],[174,27.2486],[175,27.8734],[176,28.0353],[177,27.9047],[178,27.9368],[179,27.347],[180,27.1591],[181,26.764],[182,24.9274],[184,10.253]]
       }
       body_ctrl.Scale.X {type=MovieSceneScriptingFloatChannel, default=1, const [1,182]}
       body_ctrl.Scale.Y {type=MovieSceneScriptingFloatChannel, default=1, const [1,182]}
       body_ctrl.Scale.Z {type=MovieSceneScriptingFloatChannel, default=1, const [1,182]}
       hips_ctrl.Location.X {type=MovieSceneScriptingFloatChannel, default=0, thinned_from=183, keys [[0,0],[184,0]]}
       hips_ctrl.Location.Y {type=MovieSceneScriptingFloatChannel, default=0, thinned_from=183, keys [[0,0],[184,0]]}
       hips_ctrl.Location.Z {type=MovieSceneScriptingFloatChannel, default=0, thinned_from=183, keys [[0,0],[184,0]]}
       hips_ctrl.Rotation.X {
        type=MovieSceneScriptingFloatChannel
        default=-5.577
        thinned_from=183
        keys [[0,-34.569],[13,-34.48],[27,-34.7979],[31,-35.0412],[36,-35.0708],[45,-34.9854],[46,-34.8803],[47,-34.6599],[48,-34.2808],[49,-33.8253],[50,-33.2484],[51,-32.975],[52,-32.1802],[53,-30.2484],[54,-28.9049],[55,-26.0343],[56,-24.4421],[57,-22.5094],[58,-20.7856],[59,-18.8547],[61,-16.1793],[62,-15.7977],[63,-14.5191],[64,-13.7532],[65,-13.5841],[66,-12.8452],[67,-12.2221],[68,-11.6379],[69,-11.2324],[70,-10.9187],[73,-10.6249],[74,-10.4827],[76,-10.0248],[77,-9.8359],[79,-9.6643],[80,-9.3608],[81,-9.246],[82,-9.1646],[83,-9.1109],[85,-9.0755],[87,-9.0023],[90,-9.0993],[92,-9.2075],[94,-9.2527],[96,-9.3627],[98,-9.4056],[100,-9.4886],[101,-9.4766],[102,-9.4427],[103,-9.4874],[106,-8.933],[107,-8.8714],[108,-8.7638],[109,-8.6921],[110,-8.675],[111,-8.7229],[113,-8.8856],[114,-8.913],[115,-8.9593],[116,-8.9566],[117,-8.9759],[118,-8.9609],[119,-8.9224],[123,-8.9156],[126,-9.0469],[127,-9.0621],[128,-9.1238],[129,-9.1384],[130,-9.1331],[131,-9.1585],[133,-9.2672],[134,-9.3751],[135,-9.4239],[136,-9.3762],[138,-9.2082],[139,-9.2118],[142,-9.3324],[143,-9.4055],[144,-9.4518],[147,-9.4908],[150,-9.5023],[151,-9.5231],[154,-9.6768],[156,-9.716],[158,-9.6278],[159,-9.6146],[161,-9.6849],[162,-9.7866],[163,-9.924],[164,-10.1366],[165,-10.4352],[166,-10.9374],[167,-11.9536],[168,-14.6324],[169,-18.5565],[170,-22.6611],[171,-26.349],[172,-28.5099],[173,-30.7778],[174,-32.3339],[175,-33.4271],[176,-33.806],[177,-33.5696],[178,-33.2245],[179,-32.9727],[180,-32.4392],[181,-31.1307],[182,-26.4638],[183,-15.5518],[184,-5.577]]
       }
       hips_ctrl.Rotation.Y {
        type=MovieSceneScriptingFloatChannel
        default=-4.789
        thinned_from=183
        keys [[0,4.6507],[1,4.6359],[2,4.5974],[3,4.5757],[4,4.6029],[6,4.7259],[7,4.8431],[9,5.0247],[10,5.1735],[11,5.2877],[12,5.4211],[13,5.4693],[14,5.5011],[15,5.4516],[16,5.3903],[17,5.3174],[18,5.2126],[19,5.155],[20,5.1132],[21,5.015],[23,4.9296],[24,4.8962],[25,4.8923],[26,4.8762],[27,4.9223],[28,5.0018],[29,5.0369],[30,5.0877],[32,5.2866],[33,5.3573],[35,5.4312],[36,5.4456],[40,5.3972],[41,5.3742],[42,5.334],[43,5.3341],[45,5.4156],[47,5.5912],[49,5.7233],[50,5.6995],[51,5.6464],[52,5.4407],[54,4.4254],[55,3.5867],[56,3.0608],[57,2.577],[58,2.3389],[59,2.1314],[60,1.8383],[61,1.2859],[62,1.0201],[63,0.6769],[64,0.4953],[65,0.3719],[66,0.0841],[67,-0.1424],[68,-0.4203],[69,-0.4828],[70,-0.4263],[71,-0.5457],[72,-0.5359],[73,-0.1587],[74,0.1187],[75,0.3583],[76,0.6729],[77,0.9591],[79,1.202],[80,1.4038],[81,1.6946],[82,1.9567],[83,2.1733],[84,2.286],[85,2.4039],[86,2.6242],[87,2.7194],[88,2.6852],[89,2.6134],[90,2.5817],[91,2.6162],[92,2.6304],[93,2.6254],[94,2.6146],[95,2.6327],[96,2.6734],[98,2.8143],[99,2.8703],[100,2.9338],[101,3.0178],[102,3.1466],[103,3.3612],[106,3.4314],[107,3.4335],[108,3.4451],[110,3.5778],[111,3.5443],[112,3.49],[113,3.4576],[114,3.4828],[115,3.5247],[117,3.4265],[118,3.4456],[119,3.5143],[120,3.5718],[121,3.6012],[122,3.6873],[123,3.7896],[124,3.8428],[125,3.8329],[126,3.8092],[127,3.8224],[128,3.8965],[129,3.9521],[131,3.9833],[132,4.0267],[133,4.0468],[134,4.1442],[135,4.1873],[137,4.0711],[139,4.085],[141,4.0724],[142,4.1437],[143,4.2019],[144,4.2833],[145,4.3347],[146,4.33],[148,4.3035],[149,4.3115],[150,4.3102],[151,4.3246],[152,4.3573],[153,4.4231],[154,4.5494],[156,4.7381],[158,4.8008],[159,4.8544],[160,4.9213],[161,4.9521],[162,5.007],[163,5.0223],[164,4.9158],[165,4.7601],[166,4.4623],[167,4.0959],[168,2.5954],[169,2.5432],[170,3.0154],[171,4.1658],[172,4.4053],[173,4.681],[174,4.8577],[175,5.2217],[176,5.4476],[177,5.7365],[178,6.2716],[179,6.4347],[180,6.5838],[181,6.1806],[182,0.9714],[183,-5.3726],[184,-4.789]]
       }
       hips_ctrl.Rotation.Z {
        type=MovieSceneScriptingFloatChannel
        default=0.8513
        thinned_from=183
        keys [[0,-1.9655],[1,-1.9612],[2,-1.9492],[3,-1.9583],[4,-1.9717],[5,-2.0003],[6,-2.0395],[9,-2.2909],[10,-2.3854],[11,-2.4562],[12,-2.5379],[14,-2.6162],[15,-2.5874],[16,-2.536],[17,-2.4778],[18,-2.3974],[19,-2.3382],[20,-2.3039],[21,-2.2329],[22,-2.1722],[23,-2.126],[24,-2.1048],[26,-2.1055],[27,-2.1195],[28,-2.1989],[29,-2.2615],[31,-2.3625],[32,-2.4014],[34,-2.4609],[35,-2.5032],[36,-2.4961],[37,-2.4695],[38,-2.4634],[39,-2.4498],[40,-2.4179],[41,-2.4124],[42,-2.3979],[43,-2.3776],[45,-2.3954],[46,-2.4203],[48,-2.5073],[49,-2.545],[50,-2.5295],[51,-2.4903],[52,-2.3271],[53,-1.9856],[54,-1.6898],[55,-1.2074],[56,-0.9215],[57,-0.6426],[58,-0.4762],[59,-0.3661],[60,-0.2768],[61,-0.1001],[62,-0.022],[63,0.0983],[64,0.2033],[65,0.2607],[66,0.3426],[67,0.3479],[68,0.3296],[69,0.2936],[70,0.2523],[71,0.2557],[72,0.2724],[73,0.2309],[74,0.2153],[75,0.2118],[76,0.1953],[77,0.164],[78,0.1494],[79,0.1383],[80,0.0746],[81,0.0139],[82,-0.0508],[83,-0.1056],[84,-0.1287],[85,-0.1588],[86,-0.2115],[87,-0.2368],[88,-0.2237],[89,-0.1999],[90,-0.191],[91,-0.1961],[92,-0.1965],[93,-0.1947],[94,-0.1967],[95,-0.2102],[96,-0.2389],[97,-0.281],[98,-0.3348],[99,-0.3837],[100,-0.4602],[101,-0.507],[102,-0.5387],[103,-0.5343],[106,-0.5354],[107,-0.5323],[108,-0.5263],[110,-0.5436],[111,-0.5294],[112,-0.5016],[113,-0.481],[114,-0.4736],[115,-0.4599],[116,-0.4287],[117,-0.3999],[118,-0.3892],[119,-0.3966],[120,-0.412],[121,-0.4222],[123,-0.4686],[124,-0.4775],[126,-0.4735],[127,-0.4798],[128,-0.496],[129,-0.5063],[130,-0.5038],[131,-0.5061],[132,-0.5196],[133,-0.5288],[134,-0.5614],[135,-0.5723],[136,-0.5612],[137,-0.5552],[138,-0.5515],[139,-0.5408],[140,-0.5222],[141,-0.512],[142,-0.537],[143,-0.5584],[144,-0.5851],[145,-0.6051],[146,-0.6117],[147,-0.6129],[148,-0.6112],[149,-0.6136],[150,-0.6117],[151,-0.615],[152,-0.6217],[153,-0.632],[154,-0.6558],[155,-0.6731],[156,-0.6883],[157,-0.6842],[158,-0.6829],[159,-0.6883],[160,-0.7043],[161,-0.7153],[162,-0.7432],[163,-0.7591],[164,-0.7515],[165,-0.7354],[166,-0.7243],[167,-0.7403],[168,-0.4889],[169,-0.5722],[170,-0.8829],[171,-1.2747],[172,-1.4133],[173,-1.5425],[174,-1.5445],[175,-1.6549],[176,-1.7747],[177,-1.9348],[178,-2.2246],[179,-2.3566],[180,-2.4641],[181,-2.3712],[182,-0.6276],[183,1.3753],[184,0.8513]]
       }
       hips_ctrl.Scale.X {type=MovieSceneScriptingFloatChannel, default=1, const [1,183]}
       hips_ctrl.Scale.Y {type=MovieSceneScriptingFloatChannel, default=1, const [1,183]}
       hips_ctrl.Scale.Z {type=MovieSceneScriptingFloatChannel, default=1, const [1,183]}
       hips_tan_ctrl.Location.X {
        type=MovieSceneScriptingFloatChannel
        default=-0.7493
        thinned_from=185
        keys [[0,0.7425],[1,0.7402],[3,0.7296],[4,0.7337],[6,0.7518],[7,0.7681],[9,0.7914],[10,0.8128],[11,0.8295],[12,0.8486],[13,0.8553],[14,0.8585],[16,0.8438],[17,0.8335],[18,0.8191],[20,0.8062],[21,0.7929],[24,0.7789],[25,0.7785],[26,0.776],[27,0.783],[28,0.7928],[29,0.7957],[30,0.8023],[32,0.8329],[33,0.8436],[35,0.853],[36,0.8555],[40,0.8511],[41,0.8475],[42,0.8407],[43,0.8418],[45,0.8552],[46,0.8696],[47,0.8822],[49,0.8998],[50,0.8949],[51,0.8863],[52,0.8558],[54,0.7015],[55,0.5735],[56,0.4926],[57,0.4186],[58,0.3822],[59,0.3478],[60,0.2991],[61,0.2116],[62,0.1695],[63,0.1158],[64,0.0882],[65,0.0693],[66,0.024],[67,-0.0134],[68,-0.06],[69,-0.0723],[70,-0.0651],[71,-0.0843],[72,-0.0822],[73,-0.0229],[74,0.0207],[75,0.0585],[76,0.1081],[77,0.1526],[78,0.1712],[79,0.1902],[80,0.2204],[81,0.2653],[82,0.3056],[83,0.3389],[85,0.3742],[86,0.4078],[87,0.4222],[88,0.417],[89,0.406],[90,0.401],[91,0.4062],[92,0.4082],[93,0.4074],[94,0.4055],[95,0.4081],[96,0.4142],[98,0.4355],[99,0.4438],[100,0.4532],[101,0.4659],[102,0.4862],[103,0.5205],[104,0.5989],[105,0.6351],[106,0.5316],[107,0.5319],[108,0.5337],[110,0.5544],[111,0.5494],[112,0.5412],[113,0.5364],[114,0.5406],[115,0.5473],[117,0.5326],[118,0.5357],[119,0.5464],[120,0.5554],[121,0.56],[122,0.573],[123,0.5891],[124,0.5974],[125,0.5954],[126,0.5917],[127,0.5938],[128,0.6055],[129,0.6142],[131,0.6191],[132,0.6259],[133,0.6291],[134,0.6437],[135,0.6504],[136,0.6405],[137,0.6317],[139,0.6332],[141,0.6324],[142,0.6436],[143,0.6525],[144,0.6653],[145,0.6728],[146,0.6721],[148,0.6676],[150,0.6686],[151,0.6709],[152,0.6766],[153,0.6869],[154,0.7074],[156,0.7375],[157,0.7416],[158,0.7478],[159,0.7561],[160,0.7667],[161,0.7716],[162,0.7804],[163,0.7829],[164,0.7653],[165,0.741],[166,0.6952],[167,0.6378],[168,0.4073],[169,0.4004],[170,0.4748],[171,0.6573],[172,0.6958],[173,0.7406],[174,0.7712],[175,0.8301],[176,0.8654],[177,0.9094],[178,0.9902],[179,1.0154],[180,1.0363],[181,0.9693],[182,0.1448],[183,-0.8426],[184,-0.7493]]
       }
       hips_tan_ctrl.Location.Y {
        type=MovieSceneScriptingFloatChannel
        default=1.0811
        thinned_from=185
        keys [[0,6.6993],[45,6.7329],[46,6.719],[47,6.6916],[49,6.5897],[50,6.5198],[51,6.4749],[52,6.3451],[53,6.0544],[54,5.8183],[55,5.3658],[56,5.0651],[57,4.7067],[58,4.3723],[59,4.0061],[61,3.4597],[62,3.3461],[63,3.1056],[64,2.9393],[65,2.8639],[66,2.7031],[67,2.5601],[68,2.4224],[69,2.3129],[70,2.2184],[72,2.0989],[73,2.0426],[74,1.9818],[76,1.8419],[77,1.7821],[79,1.7046],[80,1.6367],[81,1.6024],[82,1.5773],[83,1.5585],[85,1.5323],[87,1.5013],[88,1.495],[90,1.49],[92,1.4915],[94,1.4858],[96,1.4927],[98,1.4928],[100,1.5033],[101,1.5013],[102,1.4959],[103,1.5027],[104,1.5826],[105,1.6312],[106,1.417],[107,1.4074],[108,1.3907],[109,1.3798],[110,1.3771],[111,1.3845],[113,1.4095],[114,1.4136],[115,1.4206],[116,1.42],[117,1.4229],[118,1.4205],[119,1.4144],[123,1.4128],[126,1.4324],[127,1.4347],[128,1.4441],[129,1.4462],[130,1.4454],[131,1.4493],[133,1.466],[134,1.4823],[135,1.4898],[136,1.4822],[138,1.4564],[139,1.4564],[142,1.4753],[143,1.4862],[144,1.4932],[147,1.4992],[150,1.501],[151,1.5042],[154,1.5266],[156,1.5317],[157,1.5262],[158,1.5175],[159,1.5153],[161,1.5257],[162,1.5412],[163,1.5623],[164,1.5956],[165,1.642],[166,1.7196],[167,1.8774],[168,2.2926],[169,2.883],[170,3.4857],[171,4.0066],[173,4.6146],[174,4.8214],[175,4.9635],[176,5.0112],[177,4.9783],[179,4.8941],[180,4.8266],[181,4.6701],[182,4.1039],[183,2.5829],[184,1.0811]]
       }
       hips_tan_ctrl.Location.Z {
        type=MovieSceneScriptingFloatChannel
        default=-0.1086
        thinned_from=185
        keys [[0,-2.1811],[3,-2.1769],[7,-2.1803],[10,-2.1761],[13,-2.1792],[20,-2.1938],[27,-2.2084],[31,-2.2377],[35,-2.245],[37,-2.2385],[41,-2.243],[45,-2.2329],[46,-2.2218],[47,-2.1975],[48,-2.1553],[49,-2.1048],[50,-2.0406],[51,-2.0071],[52,-1.9115],[53,-1.6934],[54,-1.5417],[55,-1.2528],[56,-1.0976],[57,-0.9252],[59,-0.6401],[60,-0.546],[61,-0.4587],[62,-0.4323],[63,-0.3617],[64,-0.3202],[65,-0.3091],[66,-0.273],[67,-0.2442],[68,-0.2185],[69,-0.2007],[70,-0.1867],[72,-0.174],[73,-0.1672],[74,-0.1596],[75,-0.1499],[76,-0.1409],[77,-0.1336],[79,-0.1251],[80,-0.1161],[81,-0.113],[82,-0.1112],[83,-0.1101],[85,-0.1086],[86,-0.1081],[87,-0.1071],[89,-0.1057],[90,-0.1055],[92,-0.1065],[93,-0.106],[94,-0.1059],[96,-0.1074],[97,-0.1079],[98,-0.1086],[100,-0.111],[102,-0.1117],[103,-0.1146],[104,-0.1317],[105,-0.1419],[106,-0.1034],[107,-0.1022],[108,-0.1001],[109,-0.0993],[110,-0.0996],[111,-0.1002],[113,-0.1027],[114,-0.1035],[115,-0.1049],[116,-0.1043],[119,-0.1039],[120,-0.1045],[121,-0.1047],[122,-0.1054],[123,-0.1065],[124,-0.1077],[126,-0.1093],[127,-0.1098],[128,-0.1118],[129,-0.1127],[130,-0.1127],[131,-0.1134],[132,-0.1152],[133,-0.1165],[134,-0.1198],[135,-0.1213],[136,-0.1195],[137,-0.117],[138,-0.1154],[139,-0.1154],[140,-0.1163],[141,-0.117],[142,-0.1188],[144,-0.1229],[146,-0.1241],[149,-0.1239],[150,-0.1243],[151,-0.1249],[152,-0.1264],[153,-0.1285],[154,-0.131],[155,-0.1328],[156,-0.1342],[157,-0.1337],[158,-0.1329],[159,-0.1333],[160,-0.1349],[161,-0.1361],[162,-0.1392],[164,-0.1462],[165,-0.1514],[166,-0.1606],[167,-0.1843],[168,-0.259],[169,-0.4199],[170,-0.6345],[171,-0.8693],[172,-1.0189],[173,-1.1883],[174,-1.3112],[175,-1.4044],[176,-1.4383],[177,-1.4228],[178,-1.4115],[179,-1.3841],[180,-1.3467],[181,-1.2477],[182,-0.9016],[183,-0.3889],[184,-0.1086]]
       }
       hips_tan_ctrl.Rotation.X {type=MovieSceneScriptingFloatChannel, default=0, const [0,185]}
       hips_tan_ctrl.Rotation.Y {type=MovieSceneScriptingFloatChannel, default=0, const [0,185]}
       hips_tan_ctrl.Rotation.Z {type=MovieSceneScriptingFloatChannel, default=0, const [0,185]}
       hips_tan_ctrl.Scale.X {type=MovieSceneScriptingFloatChannel, default=1, const [1,185]}
       hips_tan_ctrl.Scale.Y {type=MovieSceneScriptingFloatChannel, default=1, const [1,185]}
       hips_tan_ctrl.Scale.Z {type=MovieSceneScriptingFloatChannel, default=1, const [1,185]}
       spine_01_ctrl.Location.X {type=MovieSceneScriptingFloatChannel, default=0, thinned_from=183, keys [[0,0],[179,0],[180,-0.0001],[182,0.0002],[184,0]]}
       spine_01_ctrl.Location.Y {
        type=MovieSceneScriptingFloatChannel
        default=0
        thinned_from=183
        keys [[0,0],[172,0],[173,-0.0027],[174,0],[175,-0.0026],[176,0],[177,0],[178,-0.0027],[179,0],[180,-0.0027],[181,-0.003],[182,-0.0045],[183,0],[184,0]]
       }
       spine_01_ctrl.Location.Z {
        type=MovieSceneScriptingFloatChannel
        default=0
        thinned_from=183
        keys [[0,0],[51,-0.0001],[55,-0.0004],[56,-0.0003],[57,-0.0004],[58,-0.0002],[59,-0.0005],[60,-0.0001],[62,0],[63,-0.0003],[64,0],[165,0],[166,-0.0001],[167,0],[170,-0.0003],[172,0],[173,0.0439],[174,0],[175,0.0439],[176,0],[177,0],[178,0.0439],[179,0],[180,0.0438],[181,0.0437],[182,0.0544],[183,-0.0001],[184,0]]
       }
       spine_01_ctrl.Rotation.X {
        type=MovieSceneScriptingFloatChannel
        default=-6.7084
        thinned_from=183
        keys [[0,-28.798],[50,-28.5391],[51,-28.3445],[52,-27.8004],[53,-26.963],[54,-25.9078],[55,-24.7032],[56,-23.3345],[58,-20.4146],[59,-19],[60,-17.668],[61,-16.4817],[62,-15.5581],[63,-14.7653],[64,-13.9187],[65,-13.1052],[69,-9.9738],[70,-9.2101],[72,-7.7302],[73,-7.0336],[74,-6.391],[75,-5.7998],[76,-5.2578],[77,-4.7574],[78,-4.3159],[79,-3.9271],[80,-3.5758],[81,-3.3098],[82,-3.1051],[85,-2.5919],[86,-2.4255],[88,-2.1162],[89,-1.9694],[90,-1.8323],[91,-1.7027],[92,-1.5812],[93,-1.4701],[94,-1.3701],[95,-1.2773],[96,-1.1973],[97,-1.1369],[98,-1.0892],[99,-1.0532],[100,-1.0478],[101,-1.0494],[102,-1.0468],[103,-1.0571],[108,-1.0668],[109,-1.0721],[110,-1.0731],[111,-1.0715],[114,-1.058],[115,-1.056],[116,-1.0485],[117,-1.0441],[118,-1.0425],[120,-1.0451],[122,-1.0405],[123,-1.0403],[125,-1.0315],[127,-1.0283],[128,-1.0187],[131,-1.0226],[135,-1.0131],[136,-1.008],[137,-1.0106],[139,-1.0097],[141,-1.006],[143,-0.9969],[144,-0.995],[148,-1.0042],[149,-1.0097],[150,-1.01],[152,-1.0046],[153,-0.9964],[154,-0.9793],[155,-0.9701],[157,-0.9555],[159,-0.9467],[160,-0.9441],[161,-0.9435],[163,-0.9474],[164,-0.9476],[165,-0.9459],[166,-0.941],[167,-0.958],[168,-1.0203],[169,-0.9814],[170,-0.938],[171,-0.8333],[172,-0.7951],[173,-0.3991],[174,-0.6832],[175,-0.3698],[176,-0.663],[177,-0.6864],[178,-0.3846],[179,-0.7225],[180,-0.4566],[181,-0.6543],[182,-1.4261],[183,-6.0929],[184,-6.7084]]
       }
       spine_01_ctrl.Rotation.Y {
        type=MovieSceneScriptingFloatChannel
        default=-0.0042
        thinned_from=183
        keys [[0,0.329],[1,0.3242],[2,0.3202],[3,0.3113],[4,0.3102],[5,0.3069],[6,0.3028],[7,0.2922],[8,0.2752],[9,0.2564],[10,0.242],[11,0.2334],[12,0.2226],[13,0.2142],[14,0.2042],[15,0.2065],[17,0.2264],[18,0.2407],[19,0.2559],[20,0.2622],[21,0.2717],[22,0.2871],[23,0.2984],[24,0.3011],[26,0.2929],[27,0.2958],[28,0.2795],[29,0.2611],[30,0.2488],[31,0.2459],[33,0.2506],[34,0.2451],[35,0.2361],[36,0.2402],[37,0.2483],[38,0.2493],[39,0.2536],[40,0.2648],[41,0.2632],[42,0.2587],[43,0.2636],[44,0.2658],[45,0.2688],[46,0.2734],[47,0.2657],[48,0.2471],[49,0.2223],[50,0.1955],[51,0.1834],[52,0.1829],[53,0.1744],[54,0.1626],[55,0.1482],[56,0.1516],[57,0.1654],[58,0.1664],[59,0.1475],[60,0.1358],[61,0.158],[62,0.1664],[63,0.17],[64,0.171],[65,0.177],[66,0.1812],[67,0.1751],[68,0.1474],[69,0.106],[70,0.0752],[71,0.0757],[72,0.0868],[73,0.075],[74,0.0607],[75,0.0542],[76,0.0485],[77,0.0345],[78,0.0262],[79,0.0218],[80,-0.0152],[81,-0.0342],[82,-0.0564],[83,-0.0782],[84,-0.0886],[85,-0.1026],[86,-0.1308],[87,-0.1478],[88,-0.1427],[89,-0.1336],[90,-0.1362],[91,-0.144],[92,-0.1495],[93,-0.1514],[94,-0.1558],[95,-0.1636],[96,-0.1747],[97,-0.1843],[98,-0.1993],[99,-0.213],[100,-0.2308],[101,-0.2434],[102,-0.2542],[103,-0.2275],[106,-0.2201],[107,-0.218],[108,-0.2153],[109,-0.2194],[110,-0.2222],[111,-0.2181],[112,-0.209],[113,-0.2004],[114,-0.1951],[115,-0.1904],[116,-0.1789],[117,-0.1687],[118,-0.1668],[119,-0.17],[120,-0.1757],[121,-0.1803],[122,-0.1859],[123,-0.1903],[124,-0.1927],[125,-0.193],[126,-0.1907],[127,-0.1935],[128,-0.1998],[129,-0.2027],[130,-0.2008],[131,-0.2014],[132,-0.2069],[133,-0.2104],[134,-0.2225],[135,-0.2312],[136,-0.2353],[138,-0.2382],[139,-0.2338],[140,-0.2267],[141,-0.2237],[142,-0.2275],[143,-0.2328],[145,-0.2493],[146,-0.2557],[147,-0.2587],[149,-0.2572],[150,-0.2546],[151,-0.2498],[152,-0.2422],[153,-0.2342],[154,-0.2239],[156,-0.2078],[158,-0.1932],[160,-0.1914],[161,-0.1916],[162,-0.1956],[163,-0.202],[164,-0.2029],[165,-0.1947],[166,-0.1904],[167,-0.1928],[168,-0.0778],[169,-0.071],[170,-0.1049],[171,-0.1224],[172,-0.1317],[173,-0.0613],[174,-0.0857],[175,-0.0259],[176,-0.0831],[177,-0.1178],[178,-0.0689],[179,-0.209],[180,-0.105],[181,-0.1369],[182,-0.1603],[183,-0.0598],[184,-0.0042]]
       }
       spine_01_ctrl.Rotation.Z {
        type=MovieSceneScriptingFloatChannel
        default=0.9333
        thinned_from=183
        keys [[0,-1.0249],[1,-1.0286],[2,-1.0211],[3,-1.0087],[5,-1.0187],[6,-1.0154],[7,-1.0086],[8,-0.9923],[9,-0.9801],[10,-0.9868],[12,-0.9902],[13,-0.9808],[14,-0.9738],[15,-0.9722],[17,-0.9911],[18,-0.9962],[19,-1.0069],[20,-1.0102],[21,-1.0179],[22,-1.0428],[23,-1.0603],[24,-1.0648],[25,-1.0717],[26,-1.0684],[27,-1.0736],[29,-1.0261],[30,-1.0238],[32,-1.0498],[33,-1.0505],[35,-1.0363],[36,-1.0464],[37,-1.0525],[38,-1.0506],[40,-1.0607],[41,-1.0498],[42,-1.0429],[43,-1.057],[44,-1.0678],[45,-1.0765],[46,-1.0895],[47,-1.091],[48,-1.0824],[49,-1.0793],[50,-1.0587],[51,-1.0409],[52,-1.0075],[53,-0.9221],[54,-0.8606],[55,-0.7634],[56,-0.6914],[57,-0.6334],[58,-0.6161],[59,-0.5403],[60,-0.4028],[61,-0.2602],[62,-0.2044],[63,-0.1606],[64,-0.1967],[65,-0.2125],[66,-0.1882],[67,-0.0864],[68,0.0671],[69,0.1489],[70,0.1934],[71,0.2458],[72,0.2286],[73,0.1478],[74,0.0577],[75,-0.036],[76,-0.1362],[77,-0.2017],[78,-0.2269],[79,-0.2585],[80,-0.2306],[81,-0.2411],[82,-0.2338],[83,-0.2289],[84,-0.2374],[85,-0.2339],[87,-0.2312],[88,-0.2405],[89,-0.2561],[90,-0.2659],[92,-0.2939],[93,-0.2987],[94,-0.2966],[95,-0.2864],[96,-0.2588],[97,-0.2164],[99,-0.0974],[100,0.0052],[101,0.0497],[102,0.0516],[103,-0.0426],[106,-0.023],[107,-0.0243],[108,-0.0327],[109,-0.0403],[110,-0.0519],[111,-0.067],[112,-0.1007],[113,-0.1309],[114,-0.1584],[115,-0.2048],[116,-0.2373],[117,-0.2636],[118,-0.287],[119,-0.3003],[120,-0.2986],[122,-0.287],[123,-0.2887],[124,-0.2968],[125,-0.3002],[126,-0.2982],[127,-0.2948],[128,-0.3001],[129,-0.3061],[130,-0.3159],[131,-0.3203],[132,-0.3206],[134,-0.3129],[135,-0.3185],[136,-0.3085],[137,-0.2889],[138,-0.2925],[139,-0.3134],[140,-0.3472],[141,-0.3679],[142,-0.3591],[143,-0.3531],[144,-0.346],[145,-0.3352],[146,-0.3256],[147,-0.3197],[148,-0.3183],[149,-0.3196],[150,-0.3239],[151,-0.326],[152,-0.3307],[153,-0.3419],[154,-0.3499],[155,-0.3566],[156,-0.3624],[157,-0.373],[158,-0.3822],[159,-0.3935],[160,-0.3981],[161,-0.3963],[162,-0.388],[163,-0.391],[164,-0.3906],[165,-0.3868],[166,-0.3462],[167,-0.3001],[168,-0.3094],[169,-0.4109],[170,-0.4737],[171,-0.7537],[172,-0.8313],[173,-0.5259],[174,-1.0354],[175,-0.5718],[176,-1.1528],[177,-1.1477],[178,-0.5548],[179,-1.1036],[180,-0.5398],[181,-0.537],[182,-0.5292],[183,0.5359],[184,0.9333]]
       }
       spine_01_ctrl.Scale.X {type=MovieSceneScriptingFloatChannel, default=1, const [1,183]}
       spine_01_ctrl.Scale.Y {type=MovieSceneScriptingFloatChannel, default=1, const [1,183]}
       spine_01_ctrl.Scale.Z {type=MovieSceneScriptingFloatChannel, default=1, const [1,183]}
       spine_02_ctrl.Location.X {
        type=MovieSceneScriptingFloatChannel
        default=0
        thinned_from=183
        keys [[0,0],[172,0],[173,0.0033],[174,0],[175,-0.0005],[176,0],[177,0],[178,0.0055],[179,0],[180,0.0101],[181,0.0134],[182,0.0195],[183,0],[184,0]]
       }
       spine_02_ctrl.Location.Y {
        type=MovieSceneScriptingFloatChannel
        default=0
        thinned_from=183
        keys [[0,0],[172,0],[173,-0.0434],[174,0],[175,-0.0397],[176,0],[177,0],[178,-0.0422],[179,0],[180,-0.0522],[181,-0.0789],[182,-0.2142],[183,0],[184,0]]
       }
       spine_02_ctrl.Location.Z {
        type=MovieSceneScriptingFloatChannel
        default=0
        thinned_from=183
        keys [[0,0],[56,-0.0002],[57,-0.0004],[58,-0.0001],[59,-0.0004],[60,-0.0001],[62,0],[63,-0.0002],[64,0],[167,0],[170,-0.0002],[172,0],[173,0.0445],[174,0],[175,0.0449],[176,0],[177,0],[178,0.0447],[179,0],[180,0.0438],[181,0.041],[182,0.0402],[183,-0.0001],[184,0]]
       }
       spine_02_ctrl.Rotation.X {
        type=MovieSceneScriptingFloatChannel
        default=-7.1335
        thinned_from=183
        keys [[0,-11.9535],[47,-11.8266],[50,-11.793],[51,-11.7227],[52,-11.5221],[53,-11.2109],[54,-10.831],[55,-10.4135],[56,-9.9105],[58,-8.8114],[60,-7.7647],[61,-7.2932],[62,-6.9473],[63,-6.6541],[64,-6.311],[65,-6.0021],[67,-5.443],[70,-4.5587],[71,-4.2758],[72,-3.9766],[73,-3.7001],[74,-3.4555],[75,-3.2377],[76,-3.042],[77,-2.8582],[78,-2.7006],[80,-2.4205],[81,-2.3255],[82,-2.2488],[84,-2.1195],[85,-2.0495],[86,-1.9847],[87,-1.9303],[88,-1.8699],[89,-1.8148],[91,-1.7172],[92,-1.6719],[93,-1.6303],[95,-1.5544],[96,-1.5204],[98,-1.4761],[99,-1.4565],[101,-1.4575],[102,-1.4549],[103,-1.466],[108,-1.4759],[109,-1.4813],[111,-1.4812],[114,-1.4689],[115,-1.4674],[116,-1.4604],[117,-1.4563],[120,-1.4583],[122,-1.4541],[123,-1.4543],[125,-1.4461],[127,-1.4433],[128,-1.4342],[131,-1.4392],[135,-1.4311],[136,-1.4261],[139,-1.4288],[141,-1.4263],[143,-1.418],[144,-1.4165],[148,-1.4269],[149,-1.4328],[151,-1.4323],[152,-1.4291],[153,-1.4215],[154,-1.4051],[157,-1.3831],[160,-1.3735],[162,-1.3752],[164,-1.3787],[165,-1.3771],[166,-1.3718],[167,-1.3883],[168,-1.4502],[169,-1.4126],[170,-1.3712],[171,-1.2722],[172,-1.2362],[173,-1.1449],[174,-1.1293],[175,-1.0739],[176,-1.1141],[177,-1.1396],[178,-1.121],[179,-1.1795],[180,-1.3169],[181,-1.8456],[182,-4.2369],[183,-6.5158],[184,-7.1335]]
       }
       spine_02_ctrl.Rotation.Y {
        type=MovieSceneScriptingFloatChannel
        default=0.0703
        thinned_from=183
        keys [[0,0.1293],[1,0.1239],[2,0.1215],[3,0.1153],[4,0.1134],[5,0.1093],[6,0.1062],[7,0.0978],[8,0.0851],[9,0.0696],[10,0.0549],[11,0.0468],[12,0.0364],[13,0.0303],[14,0.022],[15,0.0243],[16,0.0315],[17,0.039],[19,0.0636],[20,0.0689],[21,0.0761],[22,0.0856],[23,0.0928],[24,0.0942],[25,0.0887],[26,0.0853],[27,0.087],[28,0.0764],[29,0.0631],[30,0.0515],[31,0.0462],[32,0.046],[33,0.0485],[34,0.0446],[35,0.0375],[36,0.0395],[37,0.0463],[38,0.0476],[39,0.0507],[40,0.0603],[41,0.0608],[42,0.0578],[43,0.0598],[44,0.0601],[45,0.0614],[46,0.0639],[47,0.0573],[48,0.0429],[49,0.0219],[50,0.0029],[51,-0.0034],[52,0.0061],[53,0.0199],[54,0.0236],[55,0.0329],[56,0.0516],[57,0.0784],[58,0.0863],[59,0.0822],[60,0.0899],[61,0.1292],[62,0.1437],[63,0.1522],[64,0.1499],[65,0.1544],[66,0.161],[67,0.164],[68,0.1494],[69,0.1149],[70,0.0879],[71,0.0924],[72,0.102],[73,0.0844],[74,0.0633],[75,0.0497],[76,0.0368],[77,0.0184],[78,0.0084],[79,0.0019],[80,-0.032],[81,-0.051],[83,-0.0926],[84,-0.1034],[85,-0.1168],[86,-0.1443],[87,-0.1608],[88,-0.1566],[89,-0.1489],[90,-0.1523],[91,-0.1611],[92,-0.1678],[93,-0.17],[94,-0.1743],[95,-0.1814],[96,-0.1902],[97,-0.1963],[98,-0.2064],[99,-0.2154],[100,-0.2248],[101,-0.2337],[102,-0.244],[103,-0.2242],[106,-0.2152],[108,-0.2111],[109,-0.2156],[110,-0.2191],[111,-0.2163],[113,-0.2037],[114,-0.2005],[115,-0.1993],[116,-0.1905],[117,-0.1824],[118,-0.1822],[119,-0.1863],[120,-0.1917],[121,-0.1958],[122,-0.2007],[123,-0.2049],[124,-0.2079],[125,-0.2085],[126,-0.2062],[127,-0.2087],[128,-0.2153],[129,-0.2185],[130,-0.2174],[131,-0.2182],[132,-0.2238],[133,-0.227],[134,-0.2386],[135,-0.2477],[136,-0.2511],[137,-0.2512],[138,-0.2527],[139,-0.25],[140,-0.2456],[141,-0.2442],[142,-0.2472],[143,-0.2519],[145,-0.2668],[146,-0.2724],[147,-0.275],[149,-0.2735],[150,-0.2712],[151,-0.2666],[153,-0.2522],[154,-0.2424],[156,-0.227],[158,-0.2138],[159,-0.2138],[160,-0.213],[161,-0.213],[162,-0.2164],[163,-0.2232],[164,-0.2245],[165,-0.2168],[166,-0.2101],[167,-0.21],[168,-0.0998],[169,-0.1077],[170,-0.1544],[171,-0.2131],[172,-0.2388],[173,-0.2497],[174,-0.2372],[175,-0.2179],[176,-0.2576],[177,-0.2894],[178,-0.3352],[179,-0.3675],[180,-0.4143],[181,-0.4648],[182,-0.4347],[183,-0.0142],[184,0.0703]]
       }
       spine_02_ctrl.Rotation.Z {
        type=MovieSceneScriptingFloatChannel
        default=0.8379
        thinned_from=183
        keys [[0,-1.0112],[1,-1.0141],[2,-1.0062],[3,-0.9924],[5,-1],[6,-0.995],[7,-0.9842],[8,-0.963],[9,-0.9456],[10,-0.9471],[12,-0.9429],[13,-0.9311],[14,-0.9219],[15,-0.9215],[17,-0.9465],[18,-0.9561],[19,-0.9709],[20,-0.9762],[21,-0.9873],[22,-1.0159],[23,-1.0364],[25,-1.0481],[26,-1.0441],[27,-1.0493],[28,-1.0204],[29,-0.9935],[30,-0.988],[32,-1.0108],[33,-1.011],[34,-1.0031],[35,-0.993],[36,-1.0037],[37,-1.0114],[38,-1.01],[39,-1.0153],[40,-1.0237],[41,-1.0129],[42,-1.0057],[43,-1.0206],[45,-1.0399],[46,-1.0526],[47,-1.0515],[48,-1.0385],[49,-1.0297],[50,-1.0049],[51,-0.9858],[52,-0.956],[53,-0.8778],[54,-0.822],[55,-0.7353],[56,-0.672],[57,-0.6235],[58,-0.6099],[59,-0.5349],[60,-0.4005],[61,-0.2681],[62,-0.2168],[63,-0.1776],[64,-0.216],[65,-0.2339],[66,-0.2134],[67,-0.1138],[68,0.0386],[69,0.1229],[70,0.1706],[71,0.2213],[72,0.2034],[73,0.1284],[74,0.0431],[75,-0.047],[76,-0.1426],[77,-0.2032],[78,-0.2262],[79,-0.2558],[80,-0.2226],[81,-0.2279],[82,-0.2156],[83,-0.2064],[84,-0.2127],[86,-0.2006],[87,-0.1968],[88,-0.2069],[89,-0.224],[90,-0.234],[91,-0.2469],[92,-0.2603],[93,-0.265],[94,-0.2627],[95,-0.2517],[96,-0.2228],[97,-0.1789],[98,-0.1173],[99,-0.0562],[100,0.0484],[101,0.095],[102,0.0993],[103,0.0062],[106,0.0257],[107,0.0243],[108,0.0158],[109,0.0094],[110,-0.0012],[111,-0.017],[112,-0.0519],[113,-0.083],[114,-0.1106],[115,-0.1566],[116,-0.1907],[117,-0.2183],[118,-0.2416],[119,-0.2538],[120,-0.2509],[121,-0.2443],[122,-0.2371],[123,-0.2371],[124,-0.2444],[125,-0.2479],[126,-0.2464],[127,-0.2426],[128,-0.2466],[129,-0.2515],[130,-0.2614],[131,-0.2654],[132,-0.2647],[133,-0.2607],[134,-0.2544],[135,-0.2588],[136,-0.2494],[137,-0.2305],[138,-0.2339],[139,-0.2551],[140,-0.2894],[141,-0.3103],[142,-0.3004],[143,-0.2933],[144,-0.2845],[145,-0.2725],[146,-0.2624],[147,-0.2564],[148,-0.2552],[149,-0.2563],[150,-0.2609],[151,-0.2631],[152,-0.2679],[153,-0.2788],[154,-0.2861],[155,-0.2922],[156,-0.2975],[157,-0.3084],[158,-0.3178],[159,-0.3286],[160,-0.3324],[161,-0.3301],[162,-0.3208],[163,-0.3229],[164,-0.3235],[165,-0.3219],[166,-0.2843],[167,-0.2407],[168,-0.2688],[169,-0.3717],[170,-0.423],[171,-0.6867],[172,-0.7589],[173,-1.121],[174,-0.9631],[175,-1.4612],[176,-1.0741],[177,-1.0606],[178,-1.4417],[179,-0.9945],[180,-1.3122],[181,-1.115],[182,-0.2561],[183,0.4381],[184,0.8379]]
       }
       spine_02_ctrl.Scale.X {type=MovieSceneScriptingFloatChannel, default=1, const [1,183]}
       spine_02_ctrl.Scale.Y {type=MovieSceneScriptingFloatChannel, default=1, const [1,183]}
       spine_02_ctrl.Scale.Z {type=MovieSceneScriptingFloatChannel, default=1, const [1,183]}
       spine_03_ctrl.Location.X {
        type=MovieSceneScriptingFloatChannel
        default=0
        thinned_from=183
        keys [[0,0],[172,0],[173,0.0083],[174,0],[175,0.0074],[176,0],[177,0],[178,0.0122],[179,0],[180,0.0153],[181,0.017],[182,0.0176],[183,0],[184,0]]
       }
       spine_03_ctrl.Location.Y {
        type=MovieSceneScriptingFloatChannel
        default=0
        thinned_from=183
        keys [[0,0],[172,0],[173,-0.0423],[174,0],[175,-0.0396],[176,0],[177,0],[178,-0.042],[179,0],[180,-0.0503],[181,-0.072],[182,-0.1843],[183,0],[184,0]]
       }
       spine_03_ctrl.Location.Z {
        type=MovieSceneScriptingFloatChannel
        default=0
        thinned_from=183
        keys [[0,0],[49,0],[55,-0.0004],[56,-0.0003],[57,-0.0005],[58,-0.0002],[59,-0.0005],[60,-0.0001],[62,0],[63,-0.0003],[64,0],[165,0],[166,-0.0001],[167,0],[170,-0.0003],[172,0],[173,-0.0975],[174,0],[175,-0.0972],[176,0],[177,0],[178,-0.0974],[179,0],[180,-0.0984],[181,-0.1014],[182,-0.1401],[183,-0.0002],[184,0]]
       }
       spine_03_ctrl.Rotation.X {
        type=MovieSceneScriptingFloatChannel
        default=-7.8027
        thinned_from=183
        keys [[0,-11.1263],[47,-11.0325],[50,-11.0018],[51,-10.9734],[52,-10.8885],[55,-10.4479],[56,-10.2367],[61,-9.05],[62,-8.8994],[63,-8.7751],[64,-8.6021],[65,-8.4639],[67,-8.2442],[71,-7.7231],[72,-7.5728],[73,-7.438],[74,-7.3279],[77,-7.0811],[80,-6.8954],[86,-6.7018],[96,-6.4961],[99,-6.4588],[102,-6.4579],[103,-6.4697],[115,-6.4715],[127,-6.446],[152,-6.4242],[161,-6.366],[166,-6.3613],[167,-6.3767],[168,-6.4372],[170,-6.3586],[171,-6.2631],[172,-6.2276],[173,-6.5345],[174,-6.1221],[175,-6.4592],[176,-6.1084],[177,-6.1343],[178,-6.5207],[179,-6.1748],[180,-6.6226],[181,-6.8367],[182,-8.0862],[183,-7.7753],[184,-7.8027]]
       }
       spine_03_ctrl.Rotation.Y {
        type=MovieSceneScriptingFloatChannel
        default=0.1338
        thinned_from=183
        keys [[0,-0.0577],[1,-0.0636],[2,-0.0645],[3,-0.0678],[4,-0.0704],[5,-0.075],[6,-0.0768],[7,-0.0825],[8,-0.0904],[9,-0.1018],[10,-0.1159],[11,-0.1232],[12,-0.1325],[13,-0.136],[14,-0.1425],[15,-0.1404],[16,-0.1361],[17,-0.1314],[18,-0.1218],[19,-0.1129],[20,-0.109],[21,-0.1044],[22,-0.101],[23,-0.0982],[24,-0.0983],[25,-0.105],[26,-0.1077],[27,-0.1069],[28,-0.1116],[29,-0.1194],[30,-0.1298],[31,-0.137],[32,-0.1393],[33,-0.1367],[34,-0.1388],[35,-0.1438],[36,-0.1438],[37,-0.1386],[38,-0.1371],[39,-0.1352],[40,-0.1275],[41,-0.1249],[42,-0.1265],[43,-0.1274],[44,-0.129],[45,-0.1294],[46,-0.1288],[47,-0.134],[48,-0.1437],[49,-0.1603],[50,-0.1714],[51,-0.1721],[52,-0.1538],[53,-0.12],[54,-0.1028],[55,-0.0726],[56,-0.0405],[57,-0.0024],[58,0.0117],[59,0.0216],[60,0.0473],[61,0.1019],[62,0.1219],[63,0.1346],[64,0.1293],[65,0.1322],[66,0.1408],[67,0.1522],[68,0.1495],[69,0.1214],[70,0.0982],[72,0.1144],[73,0.0918],[74,0.0648],[75,0.0449],[76,0.0257],[77,0.0036],[78,-0.0077],[79,-0.0161],[80,-0.0468],[81,-0.0656],[82,-0.0849],[83,-0.1045],[84,-0.1155],[85,-0.1282],[86,-0.1548],[87,-0.1708],[88,-0.1674],[89,-0.1611],[90,-0.1653],[91,-0.175],[92,-0.1826],[93,-0.1852],[94,-0.1894],[96,-0.2024],[97,-0.2052],[98,-0.2108],[99,-0.2153],[100,-0.2172],[101,-0.2225],[102,-0.2323],[103,-0.2187],[106,-0.2082],[108,-0.2048],[109,-0.2096],[110,-0.2137],[111,-0.212],[113,-0.2042],[114,-0.2029],[115,-0.2048],[116,-0.1984],[117,-0.1924],[118,-0.1938],[120,-0.2036],[121,-0.2072],[123,-0.2155],[124,-0.2188],[125,-0.2198],[126,-0.2175],[127,-0.2197],[128,-0.2265],[129,-0.2299],[130,-0.2294],[131,-0.2306],[132,-0.236],[133,-0.239],[134,-0.25],[135,-0.2594],[136,-0.2622],[137,-0.261],[138,-0.2627],[139,-0.2614],[140,-0.2595],[141,-0.2597],[142,-0.2619],[143,-0.2661],[144,-0.273],[145,-0.2792],[146,-0.2842],[147,-0.2864],[149,-0.2848],[150,-0.2828],[151,-0.2784],[153,-0.265],[154,-0.2556],[156,-0.2407],[158,-0.2288],[159,-0.2295],[160,-0.2288],[161,-0.2287],[162,-0.2314],[163,-0.2384],[164,-0.2403],[165,-0.233],[166,-0.2244],[167,-0.2219],[168,-0.1191],[169,-0.14],[170,-0.1962],[171,-0.2921],[172,-0.3324],[173,-0.4502],[174,-0.3705],[175,-0.4806],[176,-0.4136],[177,-0.4414],[178,-0.6044],[179,-0.5047],[180,-0.6685],[181,-0.6924],[182,-0.5543],[183,0.0201],[184,0.1338]]
       }
       spine_03_ctrl.Rotation.Z {
        type=MovieSceneScriptingFloatChannel
        default=0.7395
        thinned_from=183
        keys [[0,-0.9677],[1,-0.9698],[2,-0.962],[3,-0.9471],[5,-0.9523],[6,-0.9456],[7,-0.9313],[8,-0.906],[9,-0.8841],[10,-0.8803],[12,-0.8689],[13,-0.8552],[14,-0.8439],[15,-0.8447],[17,-0.8749],[18,-0.8886],[19,-0.9068],[20,-0.9139],[21,-0.9281],[22,-0.9595],[23,-0.9821],[25,-0.9937],[26,-0.9892],[27,-0.9941],[28,-0.9619],[29,-0.9318],[30,-0.9233],[31,-0.9314],[32,-0.9421],[33,-0.9418],[34,-0.9326],[35,-0.9205],[37,-0.9408],[38,-0.9398],[39,-0.9459],[40,-0.9565],[41,-0.9462],[42,-0.9388],[43,-0.9541],[45,-0.9728],[46,-0.9849],[47,-0.9815],[48,-0.9652],[49,-0.952],[50,-0.9246],[51,-0.9054],[52,-0.8812],[53,-0.8144],[54,-0.7671],[55,-0.6948],[56,-0.6423],[57,-0.6053],[58,-0.5964],[59,-0.524],[60,-0.3943],[61,-0.2733],[62,-0.2266],[63,-0.1924],[64,-0.233],[65,-0.2529],[66,-0.2365],[67,-0.1394],[68,0.0111],[69,0.0974],[70,0.1478],[71,0.1967],[72,0.1784],[73,0.1093],[74,0.0289],[75,-0.0573],[76,-0.1481],[77,-0.2038],[78,-0.2245],[79,-0.2522],[80,-0.2141],[81,-0.2144],[82,-0.1973],[83,-0.1839],[84,-0.188],[85,-0.1796],[86,-0.1692],[87,-0.163],[88,-0.1739],[89,-0.1925],[90,-0.2026],[91,-0.2145],[92,-0.2272],[93,-0.2318],[94,-0.2294],[95,-0.2177],[96,-0.1877],[97,-0.1423],[98,-0.079],[99,-0.0166],[100,0.0898],[101,0.1381],[102,0.1446],[103,0.0531],[106,0.0727],[107,0.0711],[108,0.0626],[109,0.0573],[110,0.0478],[111,0.0313],[112,-0.0046],[113,-0.0366],[114,-0.064],[115,-0.1096],[116,-0.1449],[117,-0.1737],[118,-0.1969],[119,-0.208],[120,-0.2039],[121,-0.1967],[122,-0.1881],[123,-0.1865],[124,-0.193],[125,-0.1965],[126,-0.1953],[127,-0.1913],[128,-0.194],[129,-0.198],[130,-0.2077],[131,-0.2114],[132,-0.2098],[133,-0.2053],[134,-0.1969],[135,-0.2002],[136,-0.1916],[137,-0.1733],[138,-0.1766],[139,-0.1978],[140,-0.2326],[141,-0.2536],[142,-0.2426],[143,-0.2343],[144,-0.2241],[145,-0.2109],[146,-0.2004],[147,-0.1945],[148,-0.1934],[149,-0.1944],[150,-0.1991],[151,-0.2014],[152,-0.2062],[153,-0.2168],[154,-0.2233],[155,-0.2287],[156,-0.2334],[157,-0.2445],[158,-0.2541],[159,-0.2643],[160,-0.2673],[161,-0.2645],[162,-0.2541],[164,-0.2571],[165,-0.2576],[166,-0.2231],[167,-0.182],[168,-0.2293],[169,-0.3308],[170,-0.3698],[171,-0.6131],[172,-0.6779],[173,-0.8975],[174,-0.8756],[175,-1.158],[176,-0.9762],[177,-0.9554],[178,-1.1174],[179,-0.8699],[180,-1.0109],[181,-0.8769],[182,-0.2717],[183,0.3376],[184,0.7395]]
       }
       spine_03_ctrl.Scale.X {type=MovieSceneScriptingFloatChannel, default=1, const [1,183]}
       spine_03_ctrl.Scale.Y {type=MovieSceneScriptingFloatChannel, default=1, const [1,183]}
       spine_03_ctrl.Scale.Z {type=MovieSceneScriptingFloatChannel, default=1, const [1,183]}
       spine_04_ctrl.Location.X {type=MovieSceneScriptingFloatChannel, default=0, thinned_from=183, keys [[0,0],[184,0]]}
       spine_04_ctrl.Location.Y {type=MovieSceneScriptingFloatChannel, default=0, thinned_from=183, keys [[0,0],[181,-0.0001],[182,-0.0003],[184,0]]}
       spine_04_ctrl.Location.Z {
        type=MovieSceneScriptingFloatChannel
        default=0
        thinned_from=183
        keys [[0,0],[47,0],[48,-0.0001],[49,0],[55,-0.0005],[56,-0.0004],[57,-0.0006],[58,-0.0002],[59,-0.0006],[60,-0.0001],[62,0],[63,-0.0003],[64,0],[181,-0.0002],[182,-0.0005],[184,0]]
       }
       spine_04_ctrl.Rotation.X {
        type=MovieSceneScriptingFloatChannel
        default=-15.7998
        thinned_from=183
        keys [[0,-16.9854],[50,-16.8167],[51,-16.724],[52,-16.4614],[53,-16.0532],[54,-15.5514],[55,-14.9905],[56,-14.3331],[58,-12.9129],[59,-12.2264],[60,-11.5675],[61,-10.9698],[62,-10.5228],[63,-10.1421],[64,-9.7108],[65,-9.3137],[67,-8.5804],[70,-7.4464],[71,-7.0846],[72,-6.7086],[73,-6.3578],[74,-6.0432],[75,-5.7598],[76,-5.5037],[77,-5.265],[78,-5.0582],[80,-4.6986],[81,-4.5744],[82,-4.4758],[86,-4.1402],[89,-3.9207],[92,-3.7356],[95,-3.5856],[96,-3.5432],[99,-3.4634],[102,-3.4583],[103,-3.4708],[111,-3.4858],[118,-3.4627],[123,-3.4613],[127,-3.4495],[128,-3.4403],[135,-3.4359],[136,-3.4303],[141,-3.4301],[144,-3.4191],[151,-3.4317],[153,-3.4206],[154,-3.4042],[160,-3.372],[167,-3.3706],[168,-3.3854],[170,-3.3435],[171,-3.3035],[174,-3.2488],[176,-3.2491],[180,-3.2733],[181,-3.2856],[182,-4.6773],[183,-12.4237],[184,-15.7998]]
       }
       spine_04_ctrl.Rotation.Y {
        type=MovieSceneScriptingFloatChannel
        default=0.1348
        thinned_from=183
        keys [[0,-0.2263],[1,-0.2324],[2,-0.2323],[3,-0.2342],[4,-0.237],[5,-0.2418],[6,-0.2428],[7,-0.2468],[8,-0.2517],[9,-0.2607],[10,-0.2743],[11,-0.2808],[12,-0.2893],[13,-0.2913],[14,-0.2966],[15,-0.2947],[17,-0.2893],[18,-0.2813],[19,-0.2746],[21,-0.2686],[23,-0.2686],[24,-0.2695],[25,-0.2769],[26,-0.2792],[27,-0.279],[28,-0.2802],[29,-0.2849],[30,-0.2946],[31,-0.3027],[32,-0.3061],[33,-0.3033],[34,-0.3044],[35,-0.308],[36,-0.3092],[37,-0.3049],[39,-0.302],[40,-0.2955],[41,-0.2919],[42,-0.2926],[44,-0.2976],[45,-0.2989],[46,-0.299],[47,-0.303],[48,-0.3094],[49,-0.3226],[50,-0.3282],[51,-0.3247],[52,-0.2991],[53,-0.2505],[54,-0.222],[55,-0.1753],[56,-0.1317],[57,-0.0823],[58,-0.0593],[59,-0.0383],[60,-0.0027],[61,0.0594],[62,0.0831],[63,0.0993],[64,0.0965],[65,0.1018],[66,0.1131],[67,0.1265],[68,0.1247],[69,0.0976],[70,0.0755],[71,0.0847],[72,0.0948],[73,0.0748],[74,0.0505],[75,0.0334],[76,0.017],[77,-0.0028],[78,-0.0126],[79,-0.0193],[80,-0.0496],[81,-0.0674],[82,-0.0864],[83,-0.1057],[84,-0.1162],[85,-0.1285],[86,-0.1549],[87,-0.1706],[88,-0.1666],[89,-0.1597],[90,-0.1634],[91,-0.1726],[92,-0.1797],[93,-0.182],[94,-0.186],[95,-0.1923],[96,-0.1994],[97,-0.2028],[98,-0.2094],[99,-0.2148],[100,-0.2184],[101,-0.2245],[102,-0.2344],[103,-0.2191],[106,-0.209],[108,-0.2054],[109,-0.21],[110,-0.2139],[111,-0.212],[113,-0.2029],[114,-0.201],[115,-0.2021],[116,-0.1951],[117,-0.1886],[118,-0.1896],[119,-0.194],[120,-0.1992],[121,-0.2028],[123,-0.2112],[124,-0.2145],[125,-0.2154],[126,-0.2131],[127,-0.2155],[128,-0.2222],[129,-0.2256],[130,-0.2249],[131,-0.226],[132,-0.2315],[133,-0.2346],[134,-0.2458],[135,-0.2552],[136,-0.2581],[137,-0.2572],[138,-0.2588],[139,-0.2572],[140,-0.2547],[141,-0.2545],[142,-0.2569],[143,-0.2613],[144,-0.2684],[145,-0.2749],[146,-0.28],[147,-0.2823],[149,-0.2807],[150,-0.2788],[151,-0.2742],[153,-0.2606],[154,-0.251],[156,-0.2358],[158,-0.2234],[159,-0.2239],[160,-0.2232],[161,-0.2231],[162,-0.2235],[163,-0.2251],[164,-0.2261],[165,-0.2246],[166,-0.2223],[167,-0.2337],[168,-0.2198],[169,-0.2489],[170,-0.2937],[171,-0.3556],[172,-0.3925],[173,-0.4265],[174,-0.4463],[175,-0.4712],[177,-0.5024],[178,-0.5166],[179,-0.5227],[180,-0.5267],[181,-0.5245],[182,-0.4174],[183,0.0117],[184,0.1348]]
       }
       spine_04_ctrl.Rotation.Z {
        type=MovieSceneScriptingFloatChannel
        default=0.6513
        thinned_from=183
        keys [[0,-0.7442],[1,-0.7458],[2,-0.7382],[3,-0.723],[5,-0.7261],[6,-0.7181],[7,-0.701],[8,-0.6727],[9,-0.6475],[11,-0.6318],[12,-0.6228],[13,-0.6077],[14,-0.595],[15,-0.597],[18,-0.6478],[19,-0.6684],[20,-0.6768],[21,-0.6934],[22,-0.7268],[23,-0.751],[24,-0.758],[25,-0.7629],[26,-0.7583],[27,-0.7627],[28,-0.728],[29,-0.6958],[30,-0.6852],[31,-0.6913],[32,-0.7006],[33,-0.6996],[34,-0.6894],[35,-0.676],[37,-0.6973],[38,-0.6968],[39,-0.7034],[40,-0.7154],[41,-0.7054],[42,-0.6982],[43,-0.7139],[44,-0.7236],[45,-0.7318],[46,-0.7433],[47,-0.7383],[48,-0.72],[49,-0.7046],[50,-0.6767],[51,-0.6595],[52,-0.6438],[53,-0.5929],[54,-0.561],[55,-0.5106],[56,-0.4758],[57,-0.4568],[58,-0.4611],[59,-0.3998],[60,-0.2812],[61,-0.1741],[62,-0.1362],[63,-0.111],[64,-0.1592],[65,-0.1858],[66,-0.1777],[67,-0.0872],[68,0.0563],[69,0.1372],[70,0.1837],[71,0.2269],[72,0.2043],[73,0.1352],[74,0.0533],[75,-0.0343],[76,-0.1253],[77,-0.1812],[78,-0.2035],[79,-0.2323],[80,-0.1941],[81,-0.1926],[82,-0.1738],[83,-0.159],[84,-0.1629],[85,-0.1543],[86,-0.1426],[87,-0.1366],[88,-0.1489],[89,-0.1692],[90,-0.1808],[91,-0.1932],[92,-0.2067],[93,-0.2121],[94,-0.2106],[95,-0.1992],[96,-0.1692],[97,-0.1231],[98,-0.0589],[99,0.0044],[100,0.1122],[101,0.1618],[102,0.1697],[103,0.081],[106,0.1015],[107,0.1],[108,0.0916],[109,0.0869],[110,0.0781],[111,0.0612],[112,0.0245],[113,-0.0078],[114,-0.035],[115,-0.0803],[116,-0.1163],[117,-0.1458],[118,-0.1689],[119,-0.1794],[120,-0.1745],[121,-0.167],[122,-0.1573],[123,-0.1544],[124,-0.1604],[125,-0.1641],[126,-0.1631],[127,-0.159],[128,-0.161],[129,-0.1643],[130,-0.1738],[131,-0.1773],[132,-0.1753],[133,-0.1705],[134,-0.161],[135,-0.164],[136,-0.1563],[137,-0.1388],[138,-0.142],[139,-0.1632],[140,-0.1982],[141,-0.2194],[142,-0.2074],[143,-0.1984],[144,-0.1872],[145,-0.1734],[146,-0.1631],[147,-0.1573],[148,-0.1564],[149,-0.1571],[150,-0.1618],[151,-0.1638],[152,-0.168],[153,-0.1776],[154,-0.1826],[155,-0.1866],[156,-0.19],[157,-0.2009],[158,-0.2101],[159,-0.2197],[160,-0.2218],[161,-0.2187],[162,-0.2171],[163,-0.2166],[165,-0.2171],[166,-0.2065],[167,-0.2073],[168,-0.254],[169,-0.318],[170,-0.3574],[171,-0.4652],[172,-0.5072],[173,-0.551],[174,-0.5963],[175,-0.6286],[176,-0.628],[177,-0.6206],[179,-0.6012],[180,-0.5956],[181,-0.5732],[182,-0.2645],[183,0.2119],[184,0.6513]]
       }
       spine_04_ctrl.Scale.X {type=MovieSceneScriptingFloatChannel, default=1, const [1,183]}
       spine_04_ctrl.Scale.Y {type=MovieSceneScriptingFloatChannel, default=1, const [1,183]}
       spine_04_ctrl.Scale.Z {type=MovieSceneScriptingFloatChannel, default=1, const [1,183]}
       head_ctrl.Location.X {
        type=MovieSceneScriptingFloatChannel
        default=0
        thinned_from=180
        keys [[0,0],[172,0.0001],[173,0.0561],[174,0.0001],[175,0.0551],[176,0.0001],[177,0.0001],[178,0.0536],[179,0.0001],[180,0.0525],[181,0.0517],[184,0]]
       }
       head_ctrl.Location.Y {
        type=MovieSceneScriptingFloatChannel
        default=0
        thinned_from=180
        keys [[0,0],[49,0],[54,-0.0001],[172,-0.0001],[173,-0.0269],[174,-0.0001],[175,-0.01],[176,-0.0001],[177,-0.0002],[178,0.0041],[179,-0.0002],[180,0.0123],[181,0.0269],[184,0]]
       }
       head_ctrl.Location.Z {
        type=MovieSceneScriptingFloatChannel
        default=0
        thinned_from=180
        keys [[0,0],[47,0],[48,-0.0002],[49,0],[50,-0.0004],[51,-0.0002],[52,-0.0006],[54,-0.0008],[56,-0.0004],[57,-0.0004],[58,-0.0002],[59,-0.0002],[60,0],[62,0],[63,-0.0002],[172,-0.0001],[173,-0.4124],[174,-0.0002],[175,-0.4079],[176,-0.0002],[177,-0.0002],[178,-0.4057],[179,-0.0003],[180,-0.4027],[181,-0.3947],[184,0]]
       }
       head_ctrl.Rotation.X {
        type=MovieSceneScriptingFloatChannel
        default=-63.243
        thinned_from=180
        keys [[0,-94.9559],[11,-95.0632],[29,-93.8229],[45,-93.3067],[46,-92.9563],[49,-91.0201],[51,-89.2607],[52,-87.1405],[54,-78.7769],[55,-72.7439],[56,-67.837],[57,-62.429],[58,-57.2779],[59,-51.7254],[60,-46.8582],[61,-42.5018],[62,-39.8352],[63,-36.5066],[64,-33.4744],[65,-31.3459],[66,-28.9819],[68,-25.0216],[70,-21.2939],[71,-19.5903],[72,-17.5826],[73,-15.3794],[74,-13.488],[76,-9.9754],[77,-8.5238],[78,-7.3777],[79,-6.4071],[80,-5.2496],[81,-4.3672],[82,-3.7369],[83,-3.2777],[84,-2.8591],[85,-2.292],[86,-1.5971],[87,-1.1414],[88,-0.7829],[89,-0.5273],[90,-0.3234],[91,-0.1979],[92,-0.1057],[93,0.035],[94,0.1063],[95,0.1854],[96,0.2605],[97,0.3273],[98,0.3612],[99,0.3793],[100,0.336],[101,0.3082],[102,0.3032],[103,0.184],[106,1.3011],[107,1.4688],[108,1.8396],[109,2.1693],[110,2.4251],[111,2.6044],[112,2.6491],[113,2.6788],[114,2.745],[115,2.9467],[116,3.1299],[117,3.3662],[118,3.5929],[119,3.8104],[120,4.0003],[121,4.1331],[122,4.3234],[123,4.6073],[124,4.8002],[125,4.8531],[127,4.8228],[128,4.7761],[129,4.7399],[130,4.6889],[131,4.6086],[132,4.3982],[133,4.2312],[134,3.8304],[135,3.7688],[136,4.6059],[137,6.1255],[138,8.3179],[139,10.1326],[140,11.4944],[141,12.3989],[142,12.6242],[143,12.4393],[144,12.1704],[145,11.9576],[147,11.6084],[148,11.5551],[153,11.0428],[155,10.9727],[156,10.9686],[158,11.1044],[159,11.1965],[160,11.2527],[162,11.2672],[164,11.2449],[165,11.2076],[166,11.1073],[167,9.4895],[168,5.8787],[169,1.5185],[171,-7.6881],[172,-10.8365],[173,-13.2385],[174,-15.8746],[175,-17.0842],[176,-18.7029],[177,-19.4655],[178,-19.6373],[179,-20.5044],[180,-21.239],[181,-24.3736],[184,-63.243]]
       }
       head_ctrl.Rotation.Y {
        type=MovieSceneScriptingFloatChannel
        default=5.9945
        thinned_from=143
        keys [[0,-5.1885],[2,-5.3589],[3,-5.4568],[4,-5.4761],[5,-5.4641],[6,-5.4912],[7,-5.5052],[8,-5.5618],[9,-5.7425],[11,-6.0108],[12,-5.9844],[14,-5.8613],[15,-5.8323],[16,-5.7434],[17,-5.7102],[19,-5.411],[21,-5.2715],[22,-5.3134],[23,-5.4587],[24,-5.572],[25,-5.718],[27,-5.9732],[28,-6.1244],[29,-6.3057],[31,-6.615],[32,-6.6284],[33,-6.6221],[34,-6.5873],[36,-6.5598],[37,-6.5899],[38,-6.5729],[39,-6.4405],[40,-6.3488],[41,-6.3642],[42,-6.3495],[43,-6.4285],[44,-6.4258],[46,-6.3242],[47,-6.1787],[48,-6.0871],[49,-6.0334],[50,-5.8764],[51,-5.6298],[52,-5.2286],[53,-4.5278],[54,-3.9741],[55,-2.7684],[57,-0.5261],[58,0.7465],[59,2.1115],[60,3.0288],[81,7.3495],[99,7.8217],[100,7.554],[101,7.5231],[102,7.5835],[103,7.6229],[106,7.828],[107,7.8452],[108,7.7774],[109,7.9651],[110,8.2231],[111,8.4496],[112,8.5981],[113,8.6384],[114,8.7496],[115,9.131],[117,9.6677],[118,9.9676],[119,10.3903],[120,10.773],[121,10.9533],[122,11.3026],[123,11.6813],[124,11.9225],[125,11.9105],[129,11.7483],[131,11.57],[132,11.4164],[133,11.3317],[134,11.1326],[135,11.0479],[136,11.4417],[137,12.2107],[138,13.3833],[139,14.242],[140,14.5468],[141,14.5677],[142,14.5246],[143,14.4394],[145,14.5313],[148,14.5039],[150,14.5926],[152,14.663],[153,14.8131],[154,15.0565],[156,15.4211],[158,15.5166],[162,15.5851],[163,15.5634],[165,15.3809],[166,15.2654],[167,14.7579],[168,13.6978],[169,12.6006],[171,10.5182],[172,9.7115],[173,8.9984],[174,8.3296],[175,7.7725],[176,7.4064],[178,6.8888],[179,6.7538],[180,6.2251],[181,4.8325],[184,5.9945]]
       }
       head_ctrl.Rotation.Z {
        type=MovieSceneScriptingFloatChannel
        default=2.6627
        thinned_from=180
        keys [[0,-8.8068],[2,-8.8009],[3,-8.8643],[4,-9.0098],[5,-9.2462],[6,-9.5444],[7,-9.7859],[8,-9.9514],[9,-10.0734],[12,-10.6993],[13,-10.7281],[15,-10.4761],[16,-10.3165],[17,-10.1871],[18,-10.0089],[20,-9.9126],[21,-9.7549],[22,-9.7163],[23,-9.6353],[25,-9.3724],[26,-9.2872],[27,-9.3302],[28,-9.4381],[29,-9.4486],[30,-9.5973],[32,-10.2612],[33,-10.6119],[34,-10.6987],[36,-10.7224],[37,-10.6634],[38,-10.5632],[39,-10.4304],[40,-10.3992],[42,-10.1334],[43,-10.1995],[45,-10.4191],[46,-10.6877],[47,-10.805],[50,-10.8009],[51,-11.0129],[52,-11.5482],[53,-12.0329],[54,-12.8837],[55,-13.6191],[56,-14.5578],[57,-15.3561],[58,-16.0073],[59,-16.4293],[60,-16.6416],[61,-16.5453],[63,-16.6974],[64,-16.9008],[65,-16.9679],[66,-17.1828],[67,-16.9008],[68,-16.2168],[70,-15.3221],[71,-14.8493],[72,-14.6142],[73,-14.7187],[74,-15.0126],[75,-15.3977],[76,-15.7153],[77,-15.9284],[79,-15.8625],[80,-15.9049],[81,-15.9854],[82,-15.9988],[83,-16.0806],[84,-16.2984],[85,-16.4589],[86,-16.6771],[87,-16.7261],[89,-16.9189],[90,-16.9723],[92,-17.1813],[94,-17.1706],[95,-17.1358],[96,-17.01],[97,-16.8053],[99,-16.3478],[101,-16.3105],[102,-16.3346],[103,-16.449],[106,-16.6196],[107,-16.7497],[109,-17.4068],[110,-17.6785],[112,-18.0439],[114,-18.7265],[115,-19.3486],[116,-19.7923],[117,-20.3734],[118,-20.9131],[119,-21.373],[120,-21.6803],[122,-21.9903],[123,-22.1065],[124,-22.2732],[125,-22.5422],[127,-22.5142],[128,-22.358],[129,-22.2668],[132,-22.1432],[133,-22.1686],[134,-22.2931],[136,-22.7774],[139,-23.2741],[140,-23.383],[141,-23.3497],[144,-22.7249],[146,-22.3992],[147,-22.3499],[149,-22.3114],[152,-22.4072],[153,-22.3916],[155,-22.1846],[156,-22.0336],[160,-21.6267],[165,-21.4821],[166,-21.4217],[167,-21.3164],[168,-21.1144],[169,-21.1036],[171,-20.852],[173,-20.4945],[175,-20.4927],[176,-20.4569],[178,-20.2968],[179,-20.1803],[180,-20.6678],[181,-21.9373],[184,2.6627]]
       }
       head_ctrl.Scale.X {type=MovieSceneScriptingFloatChannel, default=1, const [1,180]}
       head_ctrl.Scale.Y {type=MovieSceneScriptingFloatChannel, default=1, const [1,180]}
       head_ctrl.Scale.Z {type=MovieSceneScriptingFloatChannel, default=1, const [1,180]}
       clavicle_l_ctrl.Location.X {type=MovieSceneScriptingFloatChannel, default=0, thinned_from=185, keys [[0,0],[184,0]]}
       clavicle_l_ctrl.Location.Y {type=MovieSceneScriptingFloatChannel, default=0, thinned_from=185, keys [[0,0],[180,0],[182,-0.0002],[184,0]]}
       clavicle_l_ctrl.Location.Z {type=MovieSceneScriptingFloatChannel, default=0, thinned_from=185, keys [[0,0],[171,-0.0001],[182,-0.0002],[184,0]]}
       clavicle_l_ctrl.Rotation.X {
        type=MovieSceneScriptingFloatChannel
        default=4.1292
        thinned_from=185
        keys [[0,16.0338],[5,16.0353],[6,16.0629],[13,16.0045],[18,15.9644],[20,15.9892],[23,15.909],[27,15.9148],[28,15.9501],[31,15.9305],[32,15.8981],[34,15.9113],[35,15.9491],[36,15.9249],[38,15.9484],[43,15.872],[44,15.902],[45,15.9805],[46,16.1194],[47,16.3063],[48,16.439],[49,16.4275],[50,16.277],[51,15.993],[52,15.581],[53,15.0465],[54,14.3948],[55,13.6315],[56,12.7919],[60,9.2525],[61,8.4021],[62,7.5974],[63,6.8634],[64,6.2256],[65,5.7007],[66,5.3052],[67,5.0558],[68,4.9689],[70,4.9612],[72,4.8507],[74,4.8206],[75,4.6958],[76,4.6507],[77,4.8977],[78,4.8289],[82,4.8519],[83,4.8679],[84,5.0437],[85,5.1694],[86,5.2432],[88,5.3046],[96,5.3377],[97,5.3939],[98,5.4308],[99,5.4434],[100,5.4288],[101,5.4582],[102,5.5181],[103,5.4537],[104,4.6197],[105,3.8923],[106,5.794],[108,5.8416],[109,5.8139],[110,5.7705],[111,5.713],[113,5.6639],[114,5.6772],[115,5.6372],[116,5.6178],[117,5.5735],[119,5.5324],[121,5.5295],[122,5.5489],[123,5.5538],[127,5.5235],[128,5.5258],[129,5.5052],[131,5.4254],[132,5.3973],[133,5.3905],[134,5.4051],[135,5.4702],[136,5.5647],[137,5.6191],[138,5.6435],[140,5.6333],[141,5.6597],[142,5.7194],[143,5.7558],[145,5.7923],[150,5.7673],[152,5.7945],[153,5.827],[154,5.896],[156,5.9385],[158,6.0158],[160,6.0389],[161,6.0369],[163,6.0002],[164,6.0058],[165,6.0406],[166,5.9811],[167,5.5435],[168,4.9546],[169,4.9687],[170,5.2008],[171,5.66],[172,5.6831],[173,5.7783],[174,5.8204],[175,5.6473],[176,5.6106],[177,5.3881],[178,5.209],[179,5.1145],[180,5.0702],[181,4.9957],[182,3.0573],[183,2.3941],[184,4.1292]]
       }
       clavicle_l_ctrl.Rotation.Y {
        type=MovieSceneScriptingFloatChannel
        default=36.106
        thinned_from=185
        keys [[0,29.0314],[1,29.1468],[4,29.1188],[5,29.1592],[6,29.1184],[8,29.1877],[10,29.1314],[11,29.0549],[17,28.959],[23,29.0832],[27,29.0146],[31,28.6339],[32,28.6405],[34,28.5694],[35,28.4901],[38,28.3749],[41,28.5884],[44,28.5591],[45,28.0866],[46,26.824],[47,24.9244],[48,22.5651],[49,19.9203],[51,14.3674],[52,11.756],[53,9.4524],[54,7.6049],[55,6.3619],[56,5.4836],[57,4.6591],[58,3.9302],[59,3.3384],[60,2.9255],[61,2.7332],[62,2.8031],[63,3.0065],[64,3.1836],[65,3.3297],[66,3.4399],[67,3.5095],[69,3.5581],[70,3.6099],[71,3.6768],[73,3.8328],[74,3.905],[75,3.9443],[76,3.9929],[77,4.1061],[79,4.083],[81,4.0342],[82,4.0007],[83,3.8646],[84,3.381],[85,3.2614],[86,3.2106],[87,3.2851],[88,3.3274],[96,3.3461],[97,3.3364],[98,3.4347],[99,3.6239],[100,3.8797],[102,3.7748],[103,3.638],[104,3.535],[105,3.4003],[106,3.1484],[107,3.0617],[108,3.0098],[109,3.0484],[110,3.0539],[111,2.98],[112,2.8984],[113,2.8481],[114,2.8882],[115,2.9935],[116,2.9969],[117,3.0115],[118,3.0168],[119,3.0139],[120,2.9924],[121,2.9873],[122,3.0055],[123,3.0625],[124,3.0625],[125,3.0369],[126,3.004],[127,2.9929],[128,3.0297],[129,3.0258],[130,3.0494],[131,3.0507],[132,3.1055],[133,3.1201],[134,3.168],[135,3.1901],[136,3.2353],[137,3.2632],[138,3.325],[139,3.331],[140,3.312],[141,3.3419],[142,3.3937],[143,3.3894],[144,3.4075],[145,3.4009],[146,3.3777],[147,3.3632],[148,3.3605],[149,3.3853],[150,3.3881],[152,3.4139],[153,3.4382],[154,3.4754],[155,3.5278],[156,3.564],[157,3.5703],[158,3.6061],[159,3.6294],[160,3.6452],[161,3.6494],[162,3.684],[163,3.6805],[164,3.631],[165,3.5694],[166,3.7194],[167,4.1719],[169,3.9363],[170,4.2756],[171,4.1487],[172,3.9349],[173,3.6785],[174,3.2086],[175,2.9033],[176,2.6655],[178,2.3804],[179,2.4294],[180,2.5741],[181,3.4239],[182,8.4497],[183,27.8996],[184,36.106]]
       }
       clavicle_l_ctrl.Rotation.Z {
        type=MovieSceneScriptingFloatChannel
        default=19.0176
        thinned_from=185
        keys [[0,16.4049],[2,16.4261],[4,16.5595],[5,16.506],[6,16.6203],[8,16.6132],[9,16.7049],[10,16.6873],[11,16.7178],[12,16.7139],[13,16.7697],[14,16.7226],[17,16.6172],[19,16.4649],[20,16.4304],[22,16.1554],[23,15.9647],[25,15.8536],[27,15.7853],[29,15.9586],[30,16.0138],[31,15.9785],[32,15.8448],[34,15.8653],[35,15.9901],[37,15.8773],[40,15.9239],[41,15.8933],[43,15.9851],[45,16.0276],[48,16.2829],[49,16.3132],[50,16.2898],[51,16.2142],[52,16.0876],[53,15.9114],[54,15.687],[55,15.4157],[57,14.7567],[60,13.7018],[61,13.3956],[62,13.1384],[63,12.9766],[64,12.9329],[65,12.9878],[66,13.1215],[67,13.3144],[68,13.5467],[70,13.9257],[72,14.3918],[74,14.7362],[75,14.9479],[76,15.0863],[77,15.0849],[78,15.2105],[82,15.2769],[83,15.142],[84,14.7805],[85,14.3124],[86,14.1002],[88,13.9759],[89,13.9361],[96,13.9303],[99,14.2176],[100,14.5667],[101,14.5639],[102,14.4208],[103,14.2128],[104,13.9219],[105,13.9227],[106,13.7549],[107,13.7082],[108,13.5446],[109,13.5077],[110,13.4293],[111,13.2659],[113,13.025],[114,12.93],[115,12.7916],[117,12.6],[118,12.5581],[119,12.5625],[121,12.5352],[122,12.5472],[123,12.5917],[125,12.4589],[126,12.417],[128,12.4336],[130,12.3716],[131,12.3065],[133,12.3331],[134,12.3864],[136,12.6304],[137,12.6806],[138,12.6733],[139,12.5945],[141,12.3855],[142,12.2571],[143,12.2807],[144,12.3462],[145,12.3613],[147,12.307],[149,12.3384],[152,12.2813],[154,12.1794],[155,12.1545],[158,12.03],[159,12.0093],[163,12.0489],[164,12.1189],[165,12.2704],[166,12.3178],[167,11.607],[168,9.9342],[169,7.5462],[170,5.3053],[171,3.1579],[172,1.887],[173,0.7151],[174,-0.1589],[175,-0.4009],[176,-0.0913],[177,0.1466],[178,0.5974],[179,1.1642],[180,2.1518],[181,3.6575],[182,8.2349],[183,11.934],[184,19.0176]]
       }
       clavicle_l_ctrl.Scale.X {type=MovieSceneScriptingFloatChannel, default=1, const [1,185]}
       clavicle_l_ctrl.Scale.Y {type=MovieSceneScriptingFloatChannel, default=1, const [1,185]}
       clavicle_l_ctrl.Scale.Z {type=MovieSceneScriptingFloatChannel, default=1, const [1,185]}
       clavicle_r_ctrl.Location.X {type=MovieSceneScriptingFloatChannel, default=0, thinned_from=182, keys [[0,0],[180,0],[182,-0.0003],[184,0]]}
       clavicle_r_ctrl.Location.Y {type=MovieSceneScriptingFloatChannel, default=0, thinned_from=182, keys [[0,0],[182,-0.0001],[184,0]]}
       clavicle_r_ctrl.Location.Z {
        type=MovieSceneScriptingFloatChannel
        default=-0.0001
        thinned_from=182
        keys [[0,-0.0001],[167,-0.0001],[170,-0.0003],[180,-0.0001],[182,-0.0004],[184,-0.0001]]
       }
       clavicle_r_ctrl.Rotation.X {
        type=MovieSceneScriptingFloatChannel
        default=-37.7702
        thinned_from=182
        keys [[0,23.1384],[15,23.0437],[20,23.1327],[28,23.1667],[31,23.1247],[38,23.191],[43,23.1036],[44,23.1474],[45,23.2621],[46,23.463],[47,23.73],[48,23.9022],[49,23.8372],[50,23.5426],[51,23.026],[52,22.2947],[53,21.3563],[54,20.2184],[55,18.8885],[56,17.4341],[59,12.9224],[60,11.4616],[61,10.0686],[62,8.7712],[63,7.5978],[64,6.5763],[65,5.7342],[66,5.0989],[67,4.6976],[68,4.5577],[70,4.5483],[71,4.4898],[72,4.4157],[74,4.3645],[75,4.2382],[76,4.1908],[77,4.5293],[78,4.4536],[82,4.4585],[83,4.47],[84,4.5229],[85,4.6169],[86,4.6885],[87,4.7192],[88,4.772],[89,4.8749],[90,4.9131],[91,4.926],[94,4.929],[97,5.0439],[98,5.1037],[99,5.1392],[100,5.0892],[101,5.1065],[102,5.1557],[108,5.2796],[109,5.2574],[111,5.1928],[113,5.1615],[114,5.1745],[116,5.1158],[117,5.0746],[119,5.0306],[121,5.0343],[122,5.0515],[127,5.0453],[128,5.0503],[129,5.0315],[130,4.9859],[132,4.9218],[133,4.9133],[134,4.9197],[135,4.9727],[136,5.0536],[137,5.0851],[138,5.0943],[139,5.0808],[140,5.0783],[141,5.1159],[142,5.1782],[143,5.2128],[145,5.2398],[150,5.2186],[152,5.2352],[153,5.268],[154,5.3256],[156,5.3585],[158,5.4293],[160,5.4391],[163,5.3952],[164,5.3935],[165,5.4108],[166,5.3741],[167,5.0667],[168,4.5919],[169,4.771],[170,5.158],[171,5.7999],[172,5.868],[174,6.0871],[175,5.9377],[176,5.8723],[177,5.6237],[178,5.4348],[179,5.3136],[180,5.2399],[181,5.1223],[182,-0.6001],[183,-25.998],[184,-37.7702]]
       }
       clavicle_r_ctrl.Rotation.Y {
        type=MovieSceneScriptingFloatChannel
        default=34.4896
        thinned_from=182
        keys [[0,31.2702],[2,31.2841],[7,31.4883],[11,31.8561],[13,31.9548],[16,32.179],[18,32.1156],[22,31.759],[28,31.4809],[39,31.6804],[42,31.8519],[43,31.988],[44,32.0513],[45,31.3543],[46,29.4447],[47,26.5878],[48,23.0187],[49,18.9768],[50,14.7331],[51,10.5585],[52,6.724],[53,3.5007],[54,1.1594],[55,-0.0287],[57,-1.081],[58,-1.5923],[59,-2.0743],[60,-2.5124],[61,-2.8918],[62,-3.198],[63,-3.4209],[64,-3.5633],[65,-3.6296],[66,-3.6245],[67,-3.5524],[68,-3.4179],[69,-3.1513],[70,-2.7834],[71,-2.3821],[72,-1.9497],[73,-1.5018],[74,-1.0655],[75,-0.7304],[76,-0.4777],[77,-0.2303],[78,-0.156],[79,-0.1469],[81,-0.1344],[82,-0.1253],[83,-0.0527],[84,-0.0439],[85,-0.1767],[86,-0.4739],[87,-0.6139],[88,-0.7423],[89,-0.821],[90,-0.8063],[91,-0.8001],[92,-0.798],[94,-0.7976],[95,-0.7067],[96,-0.5938],[97,-0.4562],[98,-0.3307],[99,-0.1696],[100,0.3941],[101,0.5058],[102,0.4661],[106,-0.5992],[107,-0.6271],[108,-0.5702],[109,-0.6063],[110,-0.6733],[111,-0.7036],[112,-0.6371],[113,-0.7695],[114,-0.9853],[115,-1.0999],[116,-0.9768],[117,-0.8121],[118,-0.7425],[119,-0.6493],[120,-0.5953],[121,-0.5904],[122,-0.5838],[123,-0.6474],[124,-0.6191],[125,-0.511],[126,-0.4277],[127,-0.4222],[128,-0.3759],[129,-0.3757],[130,-0.3916],[131,-0.3757],[132,-0.3015],[133,-0.2547],[134,-0.1531],[135,0.0515],[136,0.4824],[137,1.0253],[138,1.6083],[139,1.9245],[140,2.0372],[141,2.0675],[142,1.9104],[143,1.7304],[144,1.5681],[145,1.4515],[147,1.4097],[148,1.3995],[149,1.3762],[150,1.3838],[151,1.3473],[152,1.3027],[153,1.2504],[154,1.1565],[155,1.0732],[156,1.0024],[157,0.9595],[158,0.9194],[159,0.9102],[160,0.8906],[161,0.9089],[162,0.9175],[163,0.9431],[164,0.9477],[165,0.9159],[166,1.1755],[167,1.981],[168,1.641],[169,0.0533],[170,-1.4932],[171,-3.2908],[172,-3.4278],[173,-3.1424],[174,-3.3104],[175,-3.1534],[176,-3.0549],[177,-3.0584],[178,-3.3077],[179,-3.3476],[180,-3.4186],[181,-3.0571],[182,4.7656],[183,26.0583],[184,34.4896]]
       }
       clavicle_r_ctrl.Rotation.Z {
        type=MovieSceneScriptingFloatChannel
        default=-0.0682
        thinned_from=182
        keys [[0,18.2276],[1,18.2779],[5,18.2682],[7,18.1973],[9,18.0491],[12,17.9487],[13,17.8574],[14,17.7007],[15,17.6543],[18,17.7461],[20,17.889],[22,17.9343],[23,18.027],[24,18.0708],[25,17.9935],[27,17.9622],[28,17.9103],[30,17.6967],[31,17.7047],[32,17.7551],[34,17.7506],[35,17.644],[36,17.5886],[40,17.6297],[43,17.5095],[44,17.5738],[45,17.4417],[46,17.0751],[47,16.5253],[48,15.7945],[49,14.8917],[50,13.8775],[52,11.7583],[53,10.7747],[54,9.9227],[55,9.2632],[59,7.019],[60,6.4826],[61,5.9789],[62,5.519],[63,5.2943],[64,5.4281],[65,5.8298],[66,6.4088],[68,7.7368],[70,9.0835],[71,9.6675],[72,10.1578],[73,10.4947],[74,10.7547],[75,10.9502],[76,10.9733],[77,11.416],[78,11.302],[82,11.163],[83,10.8219],[84,10.6158],[85,10.4746],[86,10.3022],[87,10.2653],[88,10.3324],[89,10.307],[90,10.258],[94,10.2457],[95,10.1755],[96,10.0313],[98,9.8627],[99,9.7315],[100,9.2424],[101,9.0286],[102,8.8353],[106,6.8384],[107,6.5586],[108,5.8752],[109,5.5932],[110,5.684],[111,6.0021],[112,6.2548],[114,6.8917],[115,7.3334],[116,7.5545],[117,7.7968],[118,7.8907],[119,7.9177],[120,7.9219],[121,7.9457],[122,7.8349],[124,7.7866],[125,7.8092],[127,7.7656],[129,7.7723],[130,7.7458],[131,7.7795],[133,7.7555],[134,7.7255],[136,7.6275],[137,7.3942],[138,7.0632],[139,6.9173],[140,6.8238],[141,6.6621],[142,6.6036],[144,6.6401],[145,6.7323],[146,6.8069],[147,6.7573],[149,6.6938],[150,6.6338],[152,6.5646],[153,6.5431],[154,6.5442],[156,6.4771],[158,6.4808],[159,6.4555],[160,6.4129],[162,6.372],[163,6.3406],[164,6.3415],[165,6.3254],[166,6.2426],[167,5.6972],[168,3.5706],[169,1.6309],[170,-0.806],[171,-2.3412],[172,-2.9186],[173,-3.2743],[174,-2.8603],[175,-2.1539],[176,-1.7963],[177,-1.7339],[178,-1.5311],[179,-1.4327],[180,-1.1228],[181,-1.0382],[182,-1.2903],[183,0.4551],[184,-0.0682]]
       }
       clavicle_r_ctrl.Scale.X {type=MovieSceneScriptingFloatChannel, default=1, const [1,182]}
       clavicle_r_ctrl.Scale.Y {type=MovieSceneScriptingFloatChannel, default=1, const [1,182]}
       clavicle_r_ctrl.Scale.Z {type=MovieSceneScriptingFloatChannel, default=1, const [1,182]}
       neck_01_ctrl.Location.X {type=MovieSceneScriptingFloatChannel, default=0, thinned_from=180, keys [[0,0],[177,0.0002],[178,0],[181,0.0002],[184,0]]}
       neck_01_ctrl.Location.Y {type=MovieSceneScriptingFloatChannel, default=0, thinned_from=180, keys [[0,0],[177,-0.0002],[178,0],[181,-0.0002],[184,0]]}
       neck_01_ctrl.Location.Z {
        type=MovieSceneScriptingFloatChannel
        default=0
        thinned_from=180
        keys [[0,0],[31,0],[32,0.0001],[45,0],[47,0],[48,-0.0001],[49,0],[50,-0.0002],[51,0],[55,-0.0006],[56,-0.0004],[57,-0.0007],[58,-0.0003],[59,-0.0007],[60,-0.0001],[62,0],[63,-0.0003],[64,0],[139,0],[140,0.0001],[171,-0.0002],[177,-0.0002],[178,0],[179,-0.0003],[181,-0.0003],[184,0]]
       }
       neck_01_ctrl.Rotation.X {
        type=MovieSceneScriptingFloatChannel
        default=7.7773
        thinned_from=180
        keys [[0,12.5394],[1,12.5329],[2,12.4958],[3,12.5],[4,12.4751],[7,12.5265],[10,12.4191],[11,12.4057],[12,12.4639],[13,12.3848],[16,12.3139],[18,12.4625],[20,12.4998],[23,12.4146],[24,12.4444],[26,12.416],[27,12.4518],[28,12.5693],[29,12.603],[30,12.59],[32,12.6763],[33,12.6192],[35,12.5924],[36,12.6185],[37,12.5959],[38,12.5425],[39,12.5778],[40,12.5805],[41,12.6396],[42,12.6554],[43,12.5667],[44,12.6545],[45,12.655],[46,12.7126],[47,12.5245],[48,11.9899],[49,11.0602],[50,9.3936],[51,8.3061],[52,6.3539],[53,2.0112],[54,-0.3565],[55,-3.4513],[56,-4.3618],[57,-5.0309],[58,-5.0919],[59,-4.5857],[60,-4.1154],[61,-4.0709],[62,-3.7983],[63,-3.6041],[64,-3.5956],[65,-3.4129],[66,-3.3597],[67,-3.3459],[69,-3.2337],[70,-3.3984],[71,-3.6886],[72,-4.1558],[73,-4.2074],[74,-4.1279],[75,-3.8733],[76,-3.4642],[78,-2.9101],[79,-2.7111],[80,-2.1471],[81,-1.9349],[83,-1.6631],[84,-1.4204],[85,-1.1413],[86,-0.9498],[87,-0.916],[88,-0.8272],[89,-0.7632],[90,-0.6973],[91,-0.6379],[92,-0.5709],[93,-0.5126],[94,-0.4722],[95,-0.4162],[96,-0.3321],[97,-0.2552],[98,-0.1586],[99,-0.1356],[100,-0.0237],[101,-0.0509],[102,-0.1495],[103,-0.563],[106,0.7396],[107,0.8849],[108,1.1845],[109,1.4703],[110,1.6109],[111,1.5852],[112,1.5771],[113,1.6659],[114,1.7792],[115,1.8183],[116,1.8328],[117,1.9107],[118,2.0197],[119,2.047],[120,1.954],[121,1.9419],[122,1.9013],[123,1.9117],[124,1.9395],[125,1.9223],[126,1.9431],[127,1.9907],[128,1.9597],[129,1.9391],[130,1.9415],[131,1.954],[132,1.9261],[133,1.9086],[134,2.0523],[135,2.4026],[136,2.9872],[137,3.4117],[138,3.2879],[139,2.8334],[140,2.544],[141,2.657],[142,2.9382],[143,3.0125],[145,3.0581],[146,3.0511],[147,3.0323],[148,3.04],[149,3.0029],[150,3.0053],[151,3.0214],[152,3.044],[153,3.1089],[155,3.2782],[156,3.3374],[157,3.4696],[158,3.5661],[160,3.7193],[165,3.7619],[166,3.7323],[167,3.4955],[168,3.0094],[169,2.3005],[171,1.0957],[172,0.8607],[173,1.5728],[174,0.4248],[175,1.8955],[176,0.0264],[177,-0.1326],[178,1.5578],[179,-0.347],[180,1.732],[181,2.48],[184,7.7773]]
       }
       neck_01_ctrl.Rotation.Y {
        type=MovieSceneScriptingFloatChannel
        default=0.7778
        thinned_from=180
        keys [[0,5.9675],[1,5.9765],[2,5.9202],[3,5.9652],[4,5.9519],[5,6.0391],[7,6.0472],[8,5.9746],[9,5.9307],[11,5.7544],[12,5.7613],[13,5.7345],[14,5.7244],[15,5.7604],[16,5.6757],[17,5.7336],[18,5.8078],[19,5.865],[21,5.8547],[23,5.8758],[24,5.8588],[25,5.943],[26,5.9496],[27,6.0551],[28,6.0797],[29,6.0747],[30,6.1059],[32,6.0832],[34,6.0954],[35,6.0662],[36,6.0192],[37,5.9307],[40,5.9187],[41,5.9277],[42,5.8801],[43,5.8666],[44,5.8179],[46,5.8974],[47,5.8247],[48,5.8359],[49,5.8177],[51,5.5297],[52,5.4803],[53,5.4865],[54,5.7077],[55,5.7914],[56,5.9363],[57,6.1756],[58,6.2656],[59,6.4624],[61,6.6883],[62,6.8884],[63,7.2849],[64,7.1779],[65,7.1595],[66,7.3081],[67,7.1295],[68,6.5965],[70,6.1534],[71,6.0114],[72,6.3323],[73,6.9528],[74,7.4811],[75,8.0266],[76,8.6],[77,9.0787],[78,9.361],[80,9.8544],[81,10.127],[82,10.2768],[83,10.3055],[84,10.2932],[85,10.0989],[86,9.8314],[87,9.7],[88,9.6216],[89,9.4485],[90,9.3478],[91,9.2983],[92,9.0768],[94,8.972],[95,8.9684],[96,9.0097],[97,9.0321],[98,9.0348],[99,8.9869],[100,8.91],[101,8.7968],[103,8.162],[106,7.1792],[107,7.063],[108,7.1718],[109,7.3597],[110,7.4793],[111,7.4636],[112,7.3385],[113,7.2736],[114,7.3466],[115,7.3971],[116,7.5135],[117,7.5925],[118,7.7074],[119,7.7617],[120,7.7818],[121,7.8804],[122,7.9964],[123,8.084],[125,8.1445],[126,8.0216],[127,8.0005],[129,8.0108],[130,8.0016],[131,7.9598],[133,7.9891],[134,8.1461],[135,8.2379],[136,8.2318],[137,8.2617],[138,8.4671],[139,8.6052],[140,8.6625],[141,8.5946],[142,8.4448],[143,8.372],[144,8.3803],[145,8.3501],[146,8.2453],[147,8.201],[148,8.1842],[149,8.142],[151,8.1331],[152,8.179],[153,8.1848],[155,8.3794],[156,8.5173],[157,8.5756],[158,8.6154],[160,8.6584],[164,8.6146],[165,8.5885],[166,8.5074],[167,8.4675],[168,8.2207],[169,8.039],[171,7.9086],[172,7.8604],[173,6.948],[174,7.7109],[175,6.7442],[176,7.5552],[177,7.5199],[178,6.6503],[179,7.4723],[180,6.608],[181,6.6478],[184,0.7778]]
       }
       neck_01_ctrl.Rotation.Z {
        type=MovieSceneScriptingFloatChannel
        default=6.8015
        thinned_from=180
        keys [[0,-4.0613],[1,-4.0533],[3,-4.086],[5,-4.2007],[6,-4.3089],[8,-4.489],[9,-4.5472],[11,-4.7489],[12,-4.8158],[13,-4.8381],[14,-4.8048],[15,-4.7409],[16,-4.7087],[17,-4.6325],[18,-4.5365],[19,-4.4969],[20,-4.4791],[21,-4.4213],[22,-4.389],[24,-4.298],[25,-4.2193],[27,-4.1602],[28,-4.1978],[29,-4.2093],[30,-4.2498],[32,-4.4766],[33,-4.5975],[34,-4.6326],[35,-4.6565],[37,-4.6572],[38,-4.6238],[39,-4.5804],[40,-4.5654],[41,-4.5205],[42,-4.4877],[43,-4.5039],[44,-4.5506],[45,-4.5728],[47,-4.6939],[49,-4.6751],[51,-4.7851],[52,-4.7948],[53,-4.6761],[54,-4.5194],[55,-4.2734],[56,-4.1077],[57,-3.8884],[58,-3.6557],[59,-3.3724],[60,-3.1513],[61,-2.9156],[62,-2.8156],[63,-2.6915],[64,-2.7685],[65,-2.8011],[66,-2.8448],[67,-2.866],[68,-2.8658],[69,-2.8377],[70,-2.823],[71,-2.7659],[72,-2.6445],[73,-2.5631],[74,-2.5594],[75,-2.5689],[77,-2.5399],[79,-2.4687],[80,-2.4497],[81,-2.4618],[82,-2.4839],[83,-2.559],[84,-2.6725],[85,-2.8234],[86,-3.0255],[87,-3.1181],[88,-3.1919],[89,-3.3003],[91,-3.4468],[92,-3.5787],[94,-3.6597],[95,-3.6788],[96,-3.6565],[97,-3.616],[98,-3.5677],[99,-3.5384],[101,-3.6413],[102,-3.761],[103,-3.9106],[106,-4.2614],[107,-4.341],[108,-4.4083],[109,-4.4608],[111,-4.5932],[112,-4.6862],[113,-4.813],[114,-4.916],[115,-5.1257],[116,-5.2393],[117,-5.4124],[118,-5.5652],[119,-5.7297],[120,-5.8515],[123,-5.9597],[125,-6.0974],[126,-6.1296],[127,-6.1319],[128,-6.0798],[129,-6.0449],[132,-6.006],[134,-5.9999],[135,-6.0536],[137,-6.2461],[139,-6.3468],[140,-6.3696],[142,-6.3778],[143,-6.3383],[144,-6.2825],[145,-6.254],[147,-6.2466],[149,-6.2576],[153,-6.2856],[155,-6.1876],[156,-6.0994],[158,-5.9966],[160,-5.9246],[165,-5.8939],[166,-5.9003],[167,-5.8565],[168,-5.7992],[169,-5.8012],[172,-5.6954],[173,0.5743],[174,-5.6812],[175,0.5423],[176,-5.719],[177,-5.7101],[178,0.4679],[179,-5.6834],[180,0.4738],[181,0.5785],[184,6.8015]]
       }
       neck_01_ctrl.Scale.X {type=MovieSceneScriptingFloatChannel, default=1, const [1,180]}
       neck_01_ctrl.Scale.Y {type=MovieSceneScriptingFloatChannel, default=1, const [1,180]}
       neck_01_ctrl.Scale.Z {type=MovieSceneScriptingFloatChannel, default=1, const [1,180]}
       neck_02_ctrl.Location.X {type=MovieSceneScriptingFloatChannel, default=0, thinned_from=180, keys [[0,0],[184,0]]}
       neck_02_ctrl.Location.Y {
        type=MovieSceneScriptingFloatChannel
        default=0
        thinned_from=180
        keys [[0,0],[51,0],[54,-0.0002],[172,-0.0001],[173,0.1032],[174,-0.0001],[175,0.1046],[176,-0.0001],[177,-0.0001],[178,0.1053],[179,-0.0002],[180,0.1061],[181,0.1082],[184,0]]
       }
       neck_02_ctrl.Location.Z {
        type=MovieSceneScriptingFloatChannel
        default=0
        thinned_from=180
        keys [[0,0],[47,0],[48,-0.0002],[49,0],[50,-0.0003],[51,-0.0002],[52,-0.0004],[54,-0.0007],[56,-0.0004],[57,-0.0004],[58,-0.0002],[59,-0.0003],[60,0],[62,0],[63,-0.0002],[64,0],[172,-0.0001],[173,0.3083],[174,-0.0001],[175,0.3125],[176,-0.0001],[177,-0.0001],[178,0.3146],[179,-0.0002],[180,0.317],[181,0.3231],[184,0]]
       }
       neck_02_ctrl.Rotation.X {
        type=MovieSceneScriptingFloatChannel
        default=-2.0006
        thinned_from=180
        keys [[0,-2.0204],[1,-2.047],[2,-2.0661],[3,-2.0685],[6,-2.0508],[7,-2.0377],[8,-2.0133],[10,-1.9959],[11,-1.9954],[12,-1.9698],[13,-1.9363],[14,-1.9244],[16,-1.9217],[17,-1.9001],[18,-1.8494],[19,-1.8243],[20,-1.8072],[21,-1.7743],[22,-1.791],[24,-1.7981],[25,-1.8073],[26,-1.7872],[27,-1.7707],[28,-1.7438],[29,-1.7215],[30,-1.7384],[31,-1.7257],[32,-1.7014],[33,-1.6956],[35,-1.6938],[36,-1.6897],[37,-1.6894],[38,-1.6943],[39,-1.6799],[41,-1.6776],[42,-1.6485],[43,-1.6569],[45,-1.6155],[46,-1.5481],[47,-1.4679],[48,-1.4356],[49,-1.4608],[50,-1.5293],[51,-1.547],[52,-1.5471],[53,-1.4777],[54,-1.1781],[55,-0.5027],[56,-0.0117],[57,0.587],[58,1.1444],[59,1.9152],[60,2.5398],[61,3.0265],[62,3.2419],[63,3.7167],[64,3.9787],[65,4.0769],[66,4.2975],[67,4.39],[68,4.3834],[70,4.3982],[71,4.378],[72,4.3935],[73,4.6101],[74,4.7929],[75,4.9938],[76,5.2969],[77,5.4765],[78,5.638],[79,5.7348],[80,5.8051],[81,5.9796],[82,6.0639],[83,6.1129],[84,6.1418],[85,6.1862],[86,6.2688],[87,6.2936],[88,6.285],[90,6.2248],[91,6.1742],[92,6.0994],[93,6.0622],[95,5.9584],[99,5.8211],[100,5.8261],[101,5.8016],[102,5.7433],[103,5.7158],[106,6.0358],[107,6.09],[108,6.2369],[109,6.4001],[110,6.5237],[111,6.5936],[112,6.6021],[114,6.6635],[115,6.7572],[116,6.8315],[118,7.0444],[119,7.1378],[122,7.3031],[123,7.4018],[124,7.4654],[125,7.4811],[127,7.4639],[128,7.4268],[131,7.4076],[133,7.2942],[134,7.192],[135,7.1912],[136,7.4681],[137,7.9665],[138,8.6278],[139,9.1413],[140,9.5187],[141,9.7723],[142,9.8214],[145,9.5893],[147,9.474],[148,9.4621],[152,9.3443],[154,9.2663],[158,9.2964],[160,9.3422],[164,9.3421],[166,9.2947],[167,8.912],[168,8.0541],[169,6.876],[171,4.3104],[172,3.4895],[173,2.253],[174,2.1568],[175,-0.1566],[176,1.4554],[177,1.2818],[178,-1.2378],[179,1.0333],[180,-1.9489],[181,-3.4118],[184,-2.0006]]
       }
       neck_02_ctrl.Rotation.Y {
        type=MovieSceneScriptingFloatChannel
        default=3.7798
        thinned_from=180
        keys [[0,1.6273],[1,1.6184],[2,1.5912],[3,1.5785],[4,1.5752],[5,1.6042],[6,1.6115],[8,1.637],[9,1.6206],[10,1.6113],[11,1.5839],[12,1.6226],[14,1.6865],[15,1.6959],[17,1.6916],[18,1.7168],[19,1.7313],[20,1.7405],[21,1.7388],[22,1.7125],[23,1.6552],[24,1.6083],[25,1.5743],[26,1.5357],[29,1.4292],[30,1.4157],[31,1.387],[32,1.3933],[33,1.4083],[34,1.4292],[36,1.4466],[37,1.4234],[38,1.4228],[39,1.4521],[40,1.4667],[41,1.4533],[42,1.4569],[43,1.446],[44,1.4396],[45,1.4554],[46,1.4778],[47,1.5314],[48,1.6086],[49,1.7118],[50,1.8675],[51,1.9714],[52,2.0984],[53,2.354],[54,2.5483],[55,2.9316],[56,3.0792],[57,3.0972],[58,3.1686],[59,3.3078],[60,3.33],[61,3.1684],[62,3.1037],[63,3.0668],[65,2.9391],[66,2.8835],[67,2.8613],[68,2.7626],[69,2.7407],[71,2.8103],[72,2.9612],[73,3.2583],[75,3.8039],[76,4.1121],[77,4.3819],[78,4.5377],[79,4.6705],[80,4.9803],[81,5.2006],[83,5.4835],[84,5.5537],[85,5.5515],[86,5.5257],[87,5.467],[88,5.3705],[89,5.254],[91,5.1232],[92,5.0406],[93,5.0138],[94,5.0008],[96,5.0262],[97,5.0507],[98,5.1053],[99,5.1221],[100,5.0681],[101,5.0736],[102,5.0684],[103,4.8986],[106,4.7414],[108,4.7229],[111,5.0035],[112,5.0359],[113,5.033],[114,5.0533],[115,5.176],[116,5.2653],[117,5.3748],[118,5.5057],[120,5.8124],[121,5.896],[123,6.1469],[124,6.229],[125,6.263],[129,6.2083],[132,6.103],[133,6.0912],[134,6.0665],[135,6.059],[136,6.1514],[137,6.3527],[138,6.7119],[139,6.9869],[140,7.0754],[141,7.0507],[142,6.9735],[143,6.9201],[148,6.9132],[150,6.9356],[152,6.9227],[156,6.9969],[160,6.9118],[163,6.9029],[166,6.8149],[167,6.7713],[168,6.5833],[171,6.1182],[172,6.0045],[173,4.944],[174,5.7498],[175,4.6641],[176,5.6272],[177,5.6129],[178,4.7156],[179,5.5783],[180,4.7615],[181,4.6456],[184,3.7798]]
       }
       neck_02_ctrl.Rotation.Z {
        type=MovieSceneScriptingFloatChannel
        default=6.294
        thinned_from=180
        keys [[0,-4.1218],[1,-4.1242],[2,-4.1394],[3,-4.1991],[4,-4.2396],[5,-4.3222],[6,-4.436],[7,-4.5399],[9,-4.7171],[12,-4.943],[13,-4.9637],[14,-4.936],[15,-4.8922],[16,-4.7949],[17,-4.7317],[18,-4.646],[19,-4.5894],[20,-4.5578],[21,-4.4706],[23,-4.3508],[24,-4.3018],[25,-4.2627],[26,-4.2452],[27,-4.2679],[28,-4.3726],[29,-4.4385],[30,-4.5182],[31,-4.6228],[32,-4.7128],[33,-4.8321],[34,-4.8743],[35,-4.8943],[36,-4.8752],[38,-4.798],[39,-4.7323],[40,-4.7008],[41,-4.6806],[42,-4.6394],[43,-4.638],[45,-4.6696],[46,-4.7326],[47,-4.7417],[48,-4.7387],[51,-4.6623],[52,-4.614],[53,-4.4264],[54,-4.3382],[55,-4.0805],[56,-3.9404],[57,-3.7416],[58,-3.4565],[59,-3.2509],[60,-3.2105],[61,-3.2109],[62,-3.2633],[63,-3.3284],[64,-3.3189],[65,-3.3285],[66,-3.4488],[68,-3.6596],[69,-3.7029],[71,-3.6267],[72,-3.4958],[73,-3.3228],[74,-3.2089],[75,-3.1211],[76,-2.9809],[77,-2.8888],[78,-2.822],[79,-2.7461],[80,-2.7933],[81,-2.7698],[82,-2.785],[83,-2.8282],[84,-2.9194],[85,-3.0442],[86,-3.1909],[89,-3.4376],[91,-3.5662],[92,-3.6551],[93,-3.684],[94,-3.7314],[96,-3.8043],[97,-3.8318],[98,-3.8718],[99,-3.9223],[100,-4.1344],[101,-4.2201],[102,-4.2498],[103,-4.1482],[106,-4.3593],[107,-4.4129],[108,-4.5145],[109,-4.5867],[110,-4.6315],[111,-4.643],[112,-4.6374],[114,-4.7665],[115,-4.8628],[116,-4.9242],[117,-5.043],[118,-5.1509],[119,-5.2476],[120,-5.3171],[121,-5.3614],[122,-5.3847],[123,-5.3906],[124,-5.4162],[125,-5.491],[127,-5.5127],[128,-5.4569],[129,-5.4173],[131,-5.3807],[132,-5.3768],[133,-5.3954],[134,-5.4583],[136,-5.6575],[137,-5.7198],[138,-5.6658],[139,-5.5759],[140,-5.5024],[141,-5.4678],[142,-5.4702],[143,-5.4382],[144,-5.3899],[146,-5.3391],[148,-5.3427],[149,-5.3256],[152,-5.3408],[153,-5.3217],[155,-5.2283],[156,-5.1584],[160,-4.9945],[163,-4.9704],[165,-4.9744],[166,-4.9906],[167,-4.9663],[169,-4.804],[171,-4.5516],[172,-4.4818],[173,-7.4379],[174,-4.3742],[175,-7.2929],[176,-4.3451],[177,-4.3305],[178,-7.2606],[179,-4.3152],[180,-7.5717],[181,-8.443],[184,6.294]]
       }
       neck_02_ctrl.Scale.X {type=MovieSceneScriptingFloatChannel, default=1, const [1,180]}
       neck_02_ctrl.Scale.Y {type=MovieSceneScriptingFloatChannel, default=1, const [1,180]}
       neck_02_ctrl.Scale.Z {type=MovieSceneScriptingFloatChannel, default=1, const [1,180]}
       head_ik_ctrl.Location.X {
        type=MovieSceneScriptingFloatChannel
        default=-1.0689
        thinned_from=185
        keys [[0,-0.5662],[1,-0.5659],[2,-0.5567],[3,-0.5623],[4,-0.5552],[5,-0.5644],[6,-0.5545],[7,-0.5472],[8,-0.5318],[9,-0.5207],[11,-0.4725],[12,-0.4683],[13,-0.4675],[14,-0.4734],[15,-0.4859],[16,-0.4771],[17,-0.4885],[18,-0.5069],[19,-0.5181],[21,-0.5228],[22,-0.5242],[23,-0.5228],[24,-0.5198],[25,-0.5353],[26,-0.536],[27,-0.5491],[29,-0.5397],[30,-0.5391],[31,-0.522],[32,-0.5072],[33,-0.4994],[34,-0.4991],[35,-0.4946],[36,-0.4875],[37,-0.4715],[39,-0.4805],[41,-0.4845],[43,-0.4794],[44,-0.4636],[45,-0.4656],[46,-0.4628],[47,-0.4555],[48,-0.4819],[49,-0.5183],[50,-0.5597],[51,-0.5764],[52,-0.6379],[53,-0.7923],[54,-0.9104],[55,-1.0336],[57,-1.145],[58,-1.1724],[59,-1.2077],[60,-1.2157],[61,-1.2195],[62,-1.2418],[63,-1.2978],[64,-1.268],[65,-1.2534],[66,-1.2636],[67,-1.2263],[68,-1.1249],[69,-1.085],[70,-1.0547],[71,-1.0402],[72,-1.1206],[73,-1.2578],[75,-1.4824],[76,-1.5984],[77,-1.6963],[78,-1.7522],[79,-1.8026],[81,-1.9151],[82,-1.9469],[83,-1.9553],[84,-1.9453],[85,-1.8959],[86,-1.8303],[87,-1.7957],[88,-1.766],[89,-1.7178],[90,-1.6883],[91,-1.6682],[92,-1.6147],[93,-1.5991],[94,-1.5871],[95,-1.5858],[98,-1.611],[99,-1.6082],[100,-1.5847],[101,-1.5646],[102,-1.5078],[103,-1.4402],[104,-1.3999],[105,-1.3928],[106,-1.2116],[107,-1.1815],[108,-1.1828],[109,-1.2055],[110,-1.2224],[111,-1.2202],[113,-1.1691],[114,-1.1683],[115,-1.1637],[116,-1.1756],[117,-1.1754],[118,-1.1838],[119,-1.1876],[120,-1.1944],[121,-1.2123],[122,-1.2387],[123,-1.2577],[124,-1.2614],[125,-1.2618],[126,-1.2382],[127,-1.2319],[129,-1.2431],[130,-1.2396],[131,-1.2305],[133,-1.2361],[134,-1.2571],[135,-1.2543],[136,-1.2274],[137,-1.2161],[138,-1.261],[139,-1.3009],[140,-1.312],[141,-1.2871],[142,-1.2477],[143,-1.2378],[144,-1.2473],[145,-1.2498],[146,-1.2367],[149,-1.2273],[151,-1.225],[152,-1.2298],[153,-1.2318],[155,-1.2695],[156,-1.2979],[157,-1.3066],[161,-1.3147],[164,-1.3068],[165,-1.3006],[166,-1.2874],[167,-1.307],[168,-1.2865],[169,-1.2916],[170,-1.514],[171,-1.3571],[173,-1.3886],[175,-1.3653],[180,-1.3823],[181,-1.3971],[182,-1.3857],[183,-0.8872],[184,-1.0689]]
       }
       head_ik_ctrl.Location.Y {
        type=MovieSceneScriptingFloatChannel
        default=0.9436
        thinned_from=185
        keys [[0,1.4509],[1,1.4491],[2,1.4403],[4,1.4362],[5,1.4426],[7,1.4463],[10,1.4255],[11,1.4208],[12,1.4312],[13,1.4231],[15,1.4211],[16,1.4161],[18,1.446],[19,1.4539],[20,1.4571],[23,1.4507],[25,1.4546],[26,1.4524],[27,1.4602],[28,1.4766],[29,1.4806],[30,1.4785],[31,1.4828],[32,1.4914],[33,1.4848],[34,1.482],[36,1.485],[38,1.4729],[39,1.4792],[40,1.4816],[41,1.4894],[42,1.4916],[43,1.48],[44,1.4903],[45,1.4948],[46,1.5097],[47,1.4909],[48,1.4283],[49,1.3109],[50,1.0994],[51,0.971],[52,0.755],[53,0.2718],[54,0.0309],[55,-0.2851],[56,-0.3794],[57,-0.4446],[58,-0.439],[59,-0.3365],[60,-0.2477],[61,-0.2257],[62,-0.1754],[63,-0.1071],[64,-0.0877],[65,-0.0532],[66,-0.0268],[67,-0.031],[68,-0.0493],[69,-0.0529],[70,-0.0884],[71,-0.1469],[72,-0.2209],[73,-0.1952],[74,-0.1477],[75,-0.0658],[76,0.0549],[77,0.1369],[78,0.207],[79,0.2579],[80,0.3644],[81,0.425],[82,0.4587],[83,0.4894],[84,0.5383],[85,0.5897],[86,0.6291],[87,0.6351],[88,0.6493],[90,0.665],[91,0.6714],[92,0.6748],[93,0.6807],[94,0.6818],[95,0.686],[96,0.6937],[98,0.7047],[99,0.6968],[100,0.7046],[101,0.6925],[102,0.6684],[103,0.5996],[104,0.32],[105,0.1921],[106,0.8402],[107,0.8697],[108,0.9377],[109,1.0088],[110,1.0495],[111,1.0562],[112,1.0586],[113,1.0799],[114,1.1087],[115,1.1351],[116,1.1535],[117,1.1862],[118,1.2244],[119,1.2458],[120,1.2416],[121,1.2487],[122,1.2529],[124,1.2828],[126,1.2843],[127,1.2913],[128,1.2815],[129,1.2765],[131,1.2759],[132,1.2647],[133,1.2575],[134,1.275],[135,1.3374],[136,1.4639],[137,1.585],[138,1.6368],[139,1.6186],[140,1.6102],[141,1.6509],[142,1.6961],[143,1.6967],[145,1.6862],[147,1.6663],[148,1.6663],[149,1.6565],[151,1.655],[152,1.6584],[153,1.6661],[156,1.7085],[157,1.7317],[158,1.75],[160,1.7792],[164,1.7821],[165,1.7803],[166,1.7637],[167,1.6639],[168,1.4552],[169,1.1702],[170,0.5083],[171,0.6328],[172,0.4892],[173,0.3367],[174,0.2483],[175,0.1619],[176,0.0835],[177,0.0259],[178,-0.029],[179,-0.0591],[180,-0.0732],[181,-0.0766],[182,0.0904],[183,0.6368],[184,0.9436]]
       }
       head_ik_ctrl.Location.Z {
        type=MovieSceneScriptingFloatChannel
        default=0.2524
        thinned_from=185
        keys [[0,1.5175],[1,1.5156],[2,1.5089],[3,1.5108],[4,1.5085],[5,1.5132],[6,1.5149],[7,1.5205],[8,1.5191],[9,1.5142],[11,1.5105],[12,1.5192],[13,1.5124],[15,1.5083],[16,1.5028],[18,1.5237],[20,1.5303],[23,1.5178],[24,1.5199],[26,1.5159],[27,1.5213],[28,1.5354],[29,1.5398],[30,1.5375],[31,1.5422],[32,1.5504],[35,1.5413],[36,1.5443],[38,1.534],[39,1.5385],[40,1.5377],[41,1.5441],[42,1.547],[43,1.5374],[44,1.5479],[45,1.5489],[46,1.5596],[47,1.5429],[48,1.4832],[49,1.3759],[50,1.1751],[51,1.0369],[52,0.7826],[53,0.2273],[54,-0.0602],[55,-0.4073],[56,-0.4754],[57,-0.5004],[58,-0.4518],[59,-0.3373],[60,-0.2482],[61,-0.2185],[62,-0.1834],[63,-0.1512],[65,-0.111],[66,-0.0999],[67,-0.0937],[68,-0.0816],[69,-0.0726],[70,-0.0741],[71,-0.0837],[72,-0.1053],[73,-0.1111],[74,-0.1131],[75,-0.1146],[76,-0.1185],[77,-0.1281],[78,-0.1355],[80,-0.1545],[81,-0.1671],[82,-0.1748],[83,-0.1778],[84,-0.1795],[85,-0.1746],[86,-0.1651],[87,-0.1606],[88,-0.1602],[89,-0.1565],[90,-0.1548],[91,-0.1538],[92,-0.1475],[93,-0.148],[94,-0.1482],[95,-0.1502],[97,-0.1572],[98,-0.1603],[99,-0.1599],[100,-0.1576],[101,-0.1527],[102,-0.1417],[103,-0.1237],[104,-0.0928],[105,-0.0844],[106,-0.1266],[107,-0.1273],[108,-0.1372],[109,-0.1497],[110,-0.158],[111,-0.1588],[112,-0.1571],[113,-0.1582],[114,-0.1631],[115,-0.1671],[116,-0.1726],[117,-0.1792],[118,-0.187],[119,-0.1915],[120,-0.1919],[121,-0.1948],[122,-0.1984],[123,-0.2038],[124,-0.207],[125,-0.2075],[126,-0.2047],[127,-0.2054],[129,-0.2043],[130,-0.2034],[131,-0.2021],[132,-0.2],[133,-0.199],[134,-0.203],[135,-0.2143],[136,-0.2368],[137,-0.2625],[138,-0.2815],[139,-0.284],[140,-0.285],[141,-0.292],[142,-0.298],[144,-0.2966],[145,-0.2945],[147,-0.2868],[148,-0.2863],[149,-0.2838],[151,-0.283],[153,-0.287],[154,-0.2963],[155,-0.3024],[156,-0.3119],[157,-0.3186],[158,-0.3259],[160,-0.3341],[161,-0.3348],[163,-0.3344],[165,-0.3321],[166,-0.3282],[167,-0.3034],[168,-0.2553],[169,-0.2011],[170,0.0706],[171,-0.1221],[172,-0.1053],[173,-0.0907],[174,-0.0833],[175,-0.0733],[176,-0.0667],[177,-0.0605],[178,-0.0548],[179,-0.0499],[180,-0.0477],[181,-0.0491],[182,-0.0653],[183,-0.1169],[184,0.2524]]
       }
       head_ik_ctrl.Rotation.X {
        type=MovieSceneScriptingFloatChannel
        default=-63.243
        thinned_from=185
        keys [[0,-94.9559],[11,-95.0632],[29,-93.8229],[45,-93.3067],[46,-92.9563],[49,-91.0201],[51,-89.2607],[52,-87.1405],[54,-78.7769],[55,-72.7439],[56,-67.837],[57,-62.429],[58,-57.2779],[59,-51.7254],[60,-46.8582],[61,-42.5018],[62,-39.8352],[63,-36.5066],[64,-33.4744],[65,-31.3459],[66,-28.9819],[68,-25.0216],[70,-21.2939],[71,-19.5903],[72,-17.5826],[73,-15.3794],[74,-13.488],[76,-9.9754],[77,-8.5238],[78,-7.3777],[79,-6.4071],[80,-5.2496],[81,-4.3672],[82,-3.7369],[83,-3.2777],[84,-2.8591],[85,-2.292],[86,-1.5971],[87,-1.1414],[88,-0.7829],[89,-0.5273],[90,-0.3234],[91,-0.1979],[92,-0.1057],[93,0.035],[94,0.1063],[95,0.1854],[96,0.2605],[97,0.3273],[98,0.3612],[99,0.3793],[100,0.336],[101,0.3082],[102,0.3032],[103,0.184],[104,-0.6532],[105,-0.9597],[106,1.3011],[107,1.4688],[108,1.8396],[109,2.1693],[110,2.4251],[111,2.6044],[112,2.6491],[113,2.6788],[114,2.745],[115,2.9467],[116,3.1299],[117,3.3662],[118,3.5929],[119,3.8104],[120,4.0003],[121,4.1331],[122,4.3234],[123,4.6073],[124,4.8002],[125,4.8531],[127,4.8228],[128,4.7761],[129,4.7399],[130,4.6889],[131,4.6086],[132,4.3982],[133,4.2312],[134,3.8304],[135,3.7688],[136,4.6059],[137,6.1255],[138,8.3179],[139,10.1326],[140,11.4944],[141,12.3989],[142,12.6242],[143,12.4393],[144,12.1704],[145,11.9576],[147,11.6084],[148,11.5551],[153,11.0428],[155,10.9727],[156,10.9686],[158,11.1044],[159,11.1965],[160,11.2527],[162,11.2672],[164,11.2449],[165,11.2076],[166,11.1073],[167,9.4895],[168,5.8787],[169,1.5185],[170,-3.312],[171,-7.6881],[172,-10.8365],[173,-13.712],[174,-15.8746],[175,-17.5578],[176,-18.7029],[177,-19.4655],[178,-19.6373],[179,-20.5044],[180,-21.7125],[181,-24.8472],[182,-37.6234],[183,-60.0188],[184,-63.243]]
       }
       head_ik_ctrl.Rotation.Y {
        type=MovieSceneScriptingFloatChannel
        default=5.9945
        thinned_from=185
        keys [[0,-5.1885],[2,-5.3589],[3,-5.4568],[4,-5.4761],[5,-5.4641],[6,-5.4912],[7,-5.5052],[8,-5.5618],[9,-5.7425],[11,-6.0108],[12,-5.9844],[14,-5.8613],[15,-5.8323],[16,-5.7434],[17,-5.7102],[19,-5.411],[21,-5.2715],[22,-5.3134],[23,-5.4587],[24,-5.572],[25,-5.718],[27,-5.9732],[28,-6.1244],[29,-6.3057],[31,-6.615],[32,-6.6284],[33,-6.6221],[34,-6.5873],[36,-6.5598],[37,-6.5899],[38,-6.5729],[39,-6.4405],[40,-6.3488],[41,-6.3642],[42,-6.3495],[43,-6.4285],[44,-6.4258],[46,-6.3242],[47,-6.1787],[48,-6.0871],[49,-6.0334],[50,-5.8764],[51,-5.6298],[52,-5.2286],[53,-4.5278],[54,-3.9741],[55,-2.7684],[57,-0.5261],[58,0.7465],[59,2.1115],[60,3.0288],[61,3.3777],[62,3.3792],[63,3.2611],[64,3.2822],[65,3.2275],[66,3.0632],[67,3.1895],[68,3.1056],[69,2.9416],[70,3.0396],[71,3.1986],[72,3.6489],[73,4.4672],[74,5.0998],[75,5.6994],[76,6.522],[77,7.205],[79,7.9546],[80,8.5774],[81,9.1519],[82,9.5033],[83,9.8113],[84,9.918],[85,9.8936],[86,9.759],[87,9.4779],[89,8.8534],[92,7.9966],[94,7.7771],[96,7.7044],[97,7.7578],[98,7.8326],[99,7.8217],[100,7.554],[101,7.5231],[102,7.5835],[103,7.6229],[104,7.5787],[105,7.6769],[106,7.828],[107,7.8452],[108,7.7774],[109,7.9651],[110,8.2231],[111,8.4496],[112,8.5981],[113,8.6384],[114,8.7496],[115,9.131],[117,9.6677],[118,9.9676],[119,10.3903],[120,10.773],[121,10.9533],[122,11.3026],[123,11.6813],[124,11.9225],[125,11.9105],[129,11.7483],[131,11.57],[132,11.4164],[133,11.3317],[134,11.1326],[135,11.0479],[136,11.4417],[137,12.2107],[138,13.3833],[139,14.242],[140,14.5468],[141,14.5677],[142,14.5246],[143,14.4394],[145,14.5313],[148,14.5039],[150,14.5926],[152,14.663],[153,14.8131],[154,15.0565],[156,15.4211],[158,15.5166],[162,15.5851],[163,15.5634],[165,15.3809],[166,15.2654],[167,14.7579],[168,13.6978],[170,11.4966],[171,10.5182],[172,9.7115],[173,8.9984],[174,8.3296],[175,7.7725],[176,7.4064],[178,6.8888],[179,6.7538],[180,6.2251],[181,4.8325],[182,2.7168],[183,1.9432],[184,5.9945]]
       }
       head_ik_ctrl.Rotation.Z {
        type=MovieSceneScriptingFloatChannel
        default=2.6627
        thinned_from=185
        keys [[0,-8.8068],[2,-8.8009],[3,-8.8643],[4,-9.0098],[5,-9.2462],[6,-9.5444],[7,-9.7859],[8,-9.9514],[9,-10.0734],[12,-10.6993],[13,-10.7281],[15,-10.4761],[16,-10.3165],[17,-10.1871],[18,-10.0089],[20,-9.9126],[21,-9.7549],[22,-9.7163],[23,-9.6353],[25,-9.3724],[26,-9.2872],[27,-9.3302],[28,-9.4381],[29,-9.4486],[30,-9.5973],[32,-10.2612],[33,-10.6119],[34,-10.6987],[36,-10.7224],[37,-10.6634],[38,-10.5632],[39,-10.4304],[40,-10.3992],[42,-10.1334],[43,-10.1995],[45,-10.4191],[46,-10.6877],[47,-10.805],[50,-10.8009],[51,-11.0129],[52,-11.5482],[53,-12.0329],[54,-12.8837],[55,-13.6191],[56,-14.5578],[57,-15.3561],[58,-16.0073],[59,-16.4293],[60,-16.6416],[61,-16.5453],[63,-16.6974],[64,-16.9008],[65,-16.9679],[66,-17.1828],[67,-16.9008],[68,-16.2168],[70,-15.3221],[71,-14.8493],[72,-14.6142],[73,-14.7187],[74,-15.0126],[75,-15.3977],[76,-15.7153],[77,-15.9284],[79,-15.8625],[80,-15.9049],[81,-15.9854],[82,-15.9988],[83,-16.0806],[84,-16.2984],[85,-16.4589],[86,-16.6771],[87,-16.7261],[89,-16.9189],[90,-16.9723],[92,-17.1813],[94,-17.1706],[95,-17.1358],[96,-17.01],[97,-16.8053],[99,-16.3478],[101,-16.3105],[102,-16.3346],[104,-16.5751],[105,-16.6487],[106,-16.6196],[107,-16.7497],[109,-17.4068],[110,-17.6785],[112,-18.0439],[114,-18.7265],[115,-19.3486],[116,-19.7923],[117,-20.3734],[118,-20.9131],[119,-21.373],[120,-21.6803],[122,-21.9903],[123,-22.1065],[124,-22.2732],[125,-22.5422],[127,-22.5142],[128,-22.358],[129,-22.2668],[132,-22.1432],[133,-22.1686],[134,-22.2931],[136,-22.7774],[139,-23.2741],[140,-23.383],[141,-23.3497],[144,-22.7249],[146,-22.3992],[147,-22.3499],[149,-22.3114],[152,-22.4072],[153,-22.3916],[155,-22.1846],[156,-22.0336],[160,-21.6267],[165,-21.4821],[166,-21.4217],[167,-21.3164],[168,-21.1144],[169,-21.1036],[170,-20.8769],[171,-20.852],[173,-20.4945],[175,-20.4927],[176,-20.4569],[178,-20.2968],[179,-20.1803],[180,-20.6678],[181,-21.9373],[182,-16.9065],[183,-4.7607],[184,2.6627]]
       }
       head_ik_ctrl.Scale.X {type=MovieSceneScriptingFloatChannel, default=1, const [1,185]}
       head_ik_ctrl.Scale.Y {type=MovieSceneScriptingFloatChannel, default=1, const [1,185]}
       head_ik_ctrl.Scale.Z {type=MovieSceneScriptingFloatChannel, default=1, const [1,185]}
       chest_ctrl.Location.X {
        type=MovieSceneScriptingFloatChannel
        default=0.0834
        thinned_from=185
        keys [[0,-0.2878],[2,-0.2878],[3,-0.2793],[6,-0.2793],[7,-0.2731],[9,-0.2564],[10,-0.253],[11,-0.253],[12,-0.2472],[13,-0.2472],[14,-0.2383],[15,-0.2383],[16,-0.2418],[17,-0.2484],[18,-0.2539],[19,-0.2607],[20,-0.2607],[21,-0.2677],[22,-0.2771],[23,-0.2838],[24,-0.2838],[25,-0.2852],[27,-0.2852],[29,-0.2654],[30,-0.2609],[31,-0.2624],[32,-0.2651],[34,-0.2651],[35,-0.2587],[36,-0.2587],[37,-0.2655],[38,-0.2655],[39,-0.2669],[40,-0.2722],[41,-0.2722],[42,-0.2671],[43,-0.2712],[44,-0.2712],[45,-0.2763],[46,-0.28],[47,-0.2778],[48,-0.2704],[49,-0.262],[50,-0.2501],[52,-0.2348],[53,-0.2151],[54,-0.1977],[55,-0.1741],[56,-0.1602],[57,-0.1527],[58,-0.1471],[59,-0.1268],[60,-0.1015],[61,-0.0897],[63,-0.0807],[64,-0.085],[65,-0.0883],[66,-0.0868],[67,-0.0737],[68,-0.0484],[69,-0.0258],[70,-0.0111],[71,-0.0078],[72,-0.0145],[73,-0.0172],[74,-0.0188],[76,-0.0273],[77,-0.0257],[78,-0.0235],[79,-0.0232],[80,-0.007],[81,0.0001],[82,0.0093],[83,0.0181],[84,0.0219],[85,0.0277],[86,0.0388],[87,0.0456],[88,0.0435],[89,0.0395],[90,0.0403],[91,0.0429],[92,0.0447],[93,0.0454],[94,0.0474],[95,0.051],[97,0.0621],[98,0.0703],[99,0.078],[100,0.0889],[101,0.0956],[102,0.0956],[103,0.0857],[104,0.0379],[105,0.0094],[106,0.0836],[108,0.0836],[109,0.0827],[110,0.083],[111,0.0811],[112,0.0762],[113,0.0717],[114,0.0686],[115,0.0649],[116,0.0592],[117,0.0542],[118,0.0525],[119,0.0532],[121,0.0532],[122,0.0599],[124,0.0599],[125,0.0612],[127,0.0612],[128,0.0648],[129,0.0648],[130,0.0645],[133,0.0645],[134,0.073],[135,0.073],[136,0.0782],[137,0.0795],[138,0.0795],[139,0.0774],[140,0.0733],[141,0.0713],[142,0.0713],[143,0.0754],[144,0.0754],[145,0.0825],[146,0.0825],[147,0.0868],[151,0.0868],[152,0.08],[153,0.08],[154,0.0722],[155,0.0722],[156,0.0656],[157,0.0656],[158,0.0592],[159,0.0592],[160,0.0579],[161,0.0579],[162,0.0586],[163,0.0586],[164,0.0617],[165,0.0617],[166,0.0599],[167,0.0633],[168,0.0071],[169,0],[170,0.0154],[171,0.0148],[172,0.0177],[173,0.0147],[174,-0.0077],[175,-0.0167],[176,-0.0115],[177,0.005],[178,0.0312],[179,0.0497],[180,0.0673],[181,0.0924],[182,0.1028],[183,0.0649],[184,0.0834]]
       }
       chest_ctrl.Location.Y {
        type=MovieSceneScriptingFloatChannel
        default=-2.5968
        thinned_from=185
        keys [[0,-7.8653],[50,-7.8027],[51,-7.76],[52,-7.6388],[53,-7.4499],[54,-7.2086],[55,-6.9295],[56,-6.6024],[59,-5.5158],[60,-5.1679],[61,-4.8513],[62,-4.6044],[63,-4.3907],[64,-4.157],[66,-3.7169],[67,-3.5033],[70,-2.8442],[72,-2.4174],[73,-2.2146],[74,-2.0282],[75,-1.8571],[76,-1.7002],[77,-1.5547],[78,-1.4266],[79,-1.3134],[80,-1.2092],[81,-1.1317],[82,-1.0715],[84,-0.9713],[85,-0.9194],[86,-0.8701],[88,-0.7789],[89,-0.7355],[90,-0.6952],[91,-0.657],[92,-0.6211],[93,-0.5882],[94,-0.5586],[95,-0.5306],[96,-0.5063],[97,-0.4883],[98,-0.4736],[99,-0.462],[100,-0.4605],[102,-0.4609],[103,-0.4642],[104,-0.4897],[105,-0.5123],[106,-0.4663],[108,-0.4663],[109,-0.4699],[111,-0.4698],[114,-0.465],[115,-0.4644],[116,-0.4616],[117,-0.46],[118,-0.4594],[121,-0.4598],[122,-0.4588],[124,-0.4588],[125,-0.4567],[127,-0.4567],[128,-0.4508],[129,-0.4508],[130,-0.4517],[133,-0.4517],[134,-0.4499],[135,-0.4499],[136,-0.4472],[137,-0.4481],[139,-0.448],[142,-0.4468],[143,-0.4434],[146,-0.4436],[147,-0.4456],[151,-0.4456],[152,-0.4468],[153,-0.4468],[154,-0.4373],[155,-0.4373],[156,-0.4312],[157,-0.4312],[158,-0.4265],[159,-0.4265],[160,-0.4242],[163,-0.4245],[164,-0.4253],[165,-0.4253],[166,-0.4226],[167,-0.4307],[168,-0.4606],[169,-0.4419],[170,-0.4213],[171,-0.3718],[172,-0.3538],[173,-0.3244],[174,-0.3],[175,-0.2944],[176,-0.2912],[177,-0.3034],[178,-0.3148],[179,-0.3228],[180,-0.395],[181,-0.6099],[182,-1.3301],[183,-2.3694],[184,-2.5968]]
       }
       chest_ctrl.Location.Z {
        type=MovieSceneScriptingFloatChannel
        default=-0.5428
        thinned_from=185
        keys [[0,-3.5027],[34,-3.4763],[50,-3.4447],[51,-3.4063],[52,-3.3],[53,-3.1387],[54,-2.9416],[55,-2.7248],[56,-2.4841],[57,-2.2363],[58,-2.0016],[59,-1.7853],[60,-1.5891],[61,-1.4229],[62,-1.3005],[63,-1.2005],[64,-1.0943],[65,-0.9985],[66,-0.9103],[68,-0.7451],[69,-0.6683],[70,-0.5962],[71,-0.5298],[72,-0.4662],[73,-0.4097],[74,-0.3607],[75,-0.3182],[76,-0.2813],[77,-0.2488],[78,-0.2216],[79,-0.1985],[80,-0.1781],[81,-0.1636],[82,-0.1525],[85,-0.1259],[86,-0.1177],[88,-0.1029],[89,-0.0961],[90,-0.0899],[91,-0.0842],[92,-0.0789],[93,-0.0741],[94,-0.0699],[95,-0.0659],[96,-0.0626],[97,-0.0601],[98,-0.0581],[99,-0.0566],[100,-0.0565],[102,-0.0566],[103,-0.057],[104,-0.0603],[105,-0.0635],[106,-0.0572],[108,-0.0572],[109,-0.0577],[111,-0.0577],[115,-0.0569],[117,-0.0562],[124,-0.0561],[125,-0.0558],[127,-0.0558],[128,-0.055],[133,-0.0551],[135,-0.0549],[136,-0.0546],[139,-0.0547],[142,-0.0545],[143,-0.054],[146,-0.0541],[147,-0.0544],[153,-0.0545],[154,-0.0532],[155,-0.0532],[156,-0.0523],[157,-0.0523],[158,-0.0516],[159,-0.0516],[160,-0.0513],[166,-0.0515],[167,-0.0522],[168,-0.0567],[170,-0.0519],[171,-0.0449],[172,-0.0419],[173,-0.0382],[174,-0.035],[176,-0.0338],[178,-0.037],[179,-0.038],[180,-0.0476],[181,-0.0794],[182,-0.2095],[183,-0.4727],[184,-0.5428]]
       }
       chest_ctrl.Rotation.X {
        type=MovieSceneScriptingFloatChannel
        default=-21.6411
        thinned_from=185
        keys [[0,-51.8748],[47,-51.4465],[50,-51.3279],[51,-51.0346],[52,-50.2055],[53,-48.9232],[54,-47.336],[55,-45.5616],[56,-43.4797],[58,-38.9781],[59,-36.7994],[60,-34.7135],[61,-32.8264],[62,-31.4065],[63,-30.1961],[64,-28.8335],[65,-27.5731],[67,-25.2321],[70,-21.619],[72,-19.2787],[73,-18.1713],[74,-17.1744],[75,-16.274],[76,-15.4578],[77,-14.6969],[78,-14.0363],[80,-12.8913],[81,-12.4932],[82,-12.1783],[86,-11.1108],[89,-10.4116],[91,-10.0068],[93,-9.6495],[95,-9.3444],[96,-9.2119],[98,-9.0374],[99,-8.9682],[102,-8.9691],[103,-8.9933],[104,-9.1937],[105,-9.3742],[106,-9.0098],[108,-9.01],[109,-9.0383],[111,-9.0369],[118,-8.9551],[127,-8.9363],[128,-8.8874],[133,-8.8947],[135,-8.8802],[136,-8.8581],[142,-8.8549],[143,-8.8284],[146,-8.8299],[153,-8.8551],[154,-8.7814],[155,-8.7814],[156,-8.7351],[157,-8.7351],[158,-8.6984],[165,-8.6872],[166,-8.6673],[167,-8.7299],[168,-8.9649],[170,-8.6594],[171,-8.2772],[172,-8.1367],[173,-7.909],[174,-7.7259],[176,-7.6626],[178,-7.8382],[179,-7.8969],[180,-8.3034],[181,-9.4941],[182,-13.7464],[183,-20.3835],[184,-21.6411]]
       }
       chest_ctrl.Rotation.Y {
        type=MovieSceneScriptingFloatChannel
        default=0.4703
        thinned_from=185
        keys [[0,-0.7204],[2,-0.7204],[3,-0.7358],[6,-0.7358],[7,-0.7671],[8,-0.7748],[9,-0.7933],[10,-0.8294],[11,-0.8294],[12,-0.8676],[13,-0.8613],[14,-0.8819],[15,-0.8819],[16,-0.8727],[17,-0.8656],[18,-0.8467],[19,-0.8326],[20,-0.8326],[21,-0.8208],[22,-0.8284],[23,-0.8329],[24,-0.8329],[25,-0.8558],[27,-0.8558],[28,-0.8544],[29,-0.8578],[30,-0.8774],[31,-0.8932],[32,-0.9063],[34,-0.9063],[35,-0.9148],[36,-0.9148],[37,-0.9127],[38,-0.9127],[39,-0.9071],[40,-0.8909],[41,-0.8909],[42,-0.8783],[43,-0.8894],[44,-0.8894],[45,-0.9051],[46,-0.9112],[47,-0.9236],[48,-0.9463],[49,-0.9892],[50,-1.0111],[51,-1.002],[52,-0.9349],[53,-0.8035],[54,-0.7166],[55,-0.5882],[56,-0.4532],[57,-0.3098],[58,-0.2465],[59,-0.1699],[60,-0.0267],[61,0.1888],[62,0.2726],[63,0.3269],[64,0.3053],[65,0.3165],[66,0.3536],[67,0.4178],[68,0.4481],[69,0.3827],[70,0.3198],[71,0.3466],[72,0.3634],[73,0.285],[74,0.1972],[75,0.1343],[76,0.0761],[77,0.0126],[78,-0.0176],[79,-0.0383],[80,-0.1299],[81,-0.1846],[82,-0.2427],[83,-0.3018],[84,-0.3332],[85,-0.3709],[86,-0.4508],[87,-0.4985],[88,-0.4859],[89,-0.4636],[90,-0.4736],[91,-0.4998],[92,-0.5196],[93,-0.5256],[94,-0.5373],[95,-0.5568],[96,-0.5805],[97,-0.5957],[98,-0.622],[100,-0.6677],[101,-0.6916],[102,-0.6916],[103,-0.6677],[104,-0.4635],[105,-0.3365],[106,-0.6396],[108,-0.6396],[110,-0.6444],[111,-0.6451],[113,-0.6112],[114,-0.6032],[115,-0.602],[116,-0.5774],[117,-0.5549],[118,-0.5556],[119,-0.5632],[120,-0.5632],[121,-0.5765],[122,-0.6065],[123,-0.6065],[124,-0.6179],[125,-0.6249],[127,-0.6249],[128,-0.654],[129,-0.654],[130,-0.6608],[133,-0.6608],[134,-0.7238],[135,-0.7238],[136,-0.7608],[138,-0.7602],[139,-0.7579],[140,-0.7465],[141,-0.7435],[142,-0.7435],[143,-0.7654],[144,-0.7654],[145,-0.8087],[146,-0.8087],[147,-0.8325],[151,-0.8325],[152,-0.7864],[153,-0.7864],[154,-0.7357],[155,-0.7357],[156,-0.6898],[157,-0.6898],[158,-0.6512],[159,-0.6512],[160,-0.6493],[161,-0.6493],[162,-0.6545],[163,-0.6545],[164,-0.6661],[165,-0.6661],[166,-0.6326],[167,-0.6295],[168,-0.2181],[169,-0.253],[170,-0.4284],[171,-0.6592],[172,-0.7568],[173,-0.8228],[174,-0.7486],[175,-0.7817],[176,-0.8296],[177,-0.9499],[178,-1.1427],[179,-1.2429],[180,-1.3755],[181,-1.5259],[182,-1.1796],[183,0.0666],[184,0.4703]]
       }
       chest_ctrl.Rotation.Z {
        type=MovieSceneScriptingFloatChannel
        default=2.4433
        thinned_from=185
        keys [[0,-2.6688],[2,-2.6688],[3,-2.6068],[6,-2.6068],[7,-2.5694],[8,-2.5039],[9,-2.4449],[10,-2.4337],[11,-2.4337],[12,-2.4053],[13,-2.3926],[14,-2.3513],[15,-2.3513],[16,-2.3633],[17,-2.4167],[18,-2.4529],[19,-2.5024],[20,-2.5024],[21,-2.5558],[22,-2.6365],[23,-2.6947],[24,-2.6947],[25,-2.7191],[27,-2.7191],[29,-2.5576],[30,-2.5452],[31,-2.5463],[32,-2.5771],[34,-2.5771],[35,-2.5319],[36,-2.5319],[37,-2.5873],[39,-2.59],[40,-2.6328],[41,-2.6328],[42,-2.5851],[43,-2.6244],[44,-2.6244],[45,-2.6749],[46,-2.7105],[47,-2.7083],[48,-2.6573],[49,-2.616],[50,-2.5369],[52,-2.4403],[53,-2.2797],[54,-2.1636],[55,-1.9718],[56,-1.8438],[57,-1.7576],[58,-1.7439],[59,-1.5485],[60,-1.1881],[61,-0.8452],[62,-0.7078],[63,-0.5995],[64,-0.7058],[65,-0.7575],[66,-0.6995],[67,-0.4085],[68,0.0469],[69,0.3132],[70,0.4679],[71,0.6195],[72,0.5679],[73,0.3552],[74,0.1109],[75,-0.1514],[76,-0.4318],[77,-0.6083],[78,-0.6749],[79,-0.7626],[80,-0.6588],[81,-0.6726],[82,-0.6337],[83,-0.6041],[84,-0.6223],[85,-0.6037],[86,-0.5838],[87,-0.5715],[88,-0.6033],[89,-0.6563],[90,-0.6868],[91,-0.7255],[92,-0.766],[93,-0.7807],[94,-0.7744],[95,-0.7418],[96,-0.6555],[97,-0.5241],[98,-0.3396],[99,-0.1567],[100,0.1569],[101,0.2966],[102,0.2966],[103,0.0304],[104,-0.8045],[105,-1.2548],[106,0.0886],[108,0.0886],[109,0.0398],[110,0.0281],[111,-0.0392],[112,-0.1441],[113,-0.2376],[114,-0.3202],[115,-0.4582],[116,-0.5606],[117,-0.6437],[118,-0.7135],[119,-0.739],[121,-0.7382],[122,-0.7079],[123,-0.7079],[124,-0.7037],[125,-0.7169],[127,-0.7169],[128,-0.7269],[129,-0.7269],[130,-0.771],[133,-0.771],[134,-0.7491],[135,-0.7491],[136,-0.7337],[137,-0.6771],[138,-0.6771],[139,-0.7506],[140,-0.8537],[141,-0.9163],[142,-0.9163],[143,-0.8649],[144,-0.8649],[145,-0.802],[146,-0.802],[147,-0.7534],[151,-0.7534],[152,-0.7886],[153,-0.7886],[154,-0.8444],[155,-0.8444],[156,-0.8793],[157,-0.8793],[158,-0.9411],[159,-0.9411],[160,-0.9848],[161,-0.9848],[162,-0.9598],[163,-0.9598],[164,-0.9523],[165,-0.9523],[166,-0.811],[167,-0.6451],[168,-0.7598],[169,-1.1462],[170,-1.3364],[171,-2.329],[172,-2.5993],[173,-2.952],[174,-3.3679],[175,-3.7717],[176,-3.7832],[177,-3.7315],[178,-3.6691],[179,-3.4788],[180,-3.3435],[181,-2.9033],[182,-0.9902],[183,1.2981],[184,2.4433]]
       }
       chest_ctrl.Scale.X {type=MovieSceneScriptingFloatChannel, default=1, const [1,185]}
       chest_ctrl.Scale.Y {type=MovieSceneScriptingFloatChannel, default=1, const [1,185]}
       chest_ctrl.Scale.Z {type=MovieSceneScriptingFloatChannel, default=1, const [1,185]}
       chest_tan_ctrl.Location.X {
        type=MovieSceneScriptingFloatChannel
        default=-0.0086
        thinned_from=185
        keys [[0,0.0185],[2,0.0185],[3,0.0195],[6,0.0195],[7,0.0213],[8,0.022],[9,0.0233],[10,0.025],[11,0.025],[12,0.0271],[13,0.0279],[14,0.0268],[15,0.0268],[16,0.0262],[17,0.027],[19,0.0249],[20,0.0249],[21,0.024],[24,0.0238],[25,0.0247],[27,0.0247],[29,0.0259],[30,0.0253],[31,0.0283],[34,0.0284],[35,0.0289],[36,0.0289],[37,0.0285],[38,0.0285],[39,0.0282],[40,0.0272],[41,0.0272],[42,0.0269],[43,0.0272],[44,0.0272],[45,0.0276],[46,0.0277],[47,0.0274],[49,0.0314],[50,0.0325],[51,0.0322],[52,0.0293],[53,0.0238],[54,0.0205],[56,0.0099],[57,0.0039],[58,0.0013],[59,-0.001],[60,-0.0053],[61,-0.0131],[62,-0.0161],[63,-0.0181],[64,-0.0176],[65,-0.0182],[67,-0.021],[68,-0.0201],[69,-0.0159],[70,-0.0124],[71,-0.0131],[72,-0.0142],[73,-0.0116],[74,-0.0086],[75,-0.0066],[76,-0.0049],[77,-0.0026],[78,-0.0014],[79,-0.0006],[80,0.0039],[81,0.0064],[83,0.012],[84,0.0134],[85,0.0152],[86,0.0189],[87,0.0211],[88,0.0205],[89,0.0193],[90,0.0197],[91,0.0208],[92,0.0217],[93,0.0219],[94,0.0225],[95,0.0235],[96,0.0248],[97,0.0258],[99,0.0289],[100,0.0307],[101,0.0321],[102,0.0321],[103,0.0304],[104,0.0192],[105,0.0125],[106,0.0293],[110,0.0293],[111,0.0292],[113,0.0272],[115,0.0262],[117,0.0237],[120,0.0235],[121,0.0257],[122,0.0251],[123,0.0251],[124,0.0272],[125,0.0275],[127,0.0275],[128,0.028],[129,0.028],[130,0.0282],[133,0.0282],[134,0.0311],[135,0.0311],[136,0.0329],[138,0.033],[139,0.0327],[140,0.0319],[141,0.0317],[142,0.0317],[143,0.0328],[144,0.0328],[145,0.0349],[146,0.0349],[147,0.0361],[151,0.0361],[152,0.0339],[153,0.0339],[154,0.0315],[155,0.0315],[156,0.0294],[157,0.0294],[158,0.0275],[161,0.0273],[162,0.0276],[163,0.0276],[164,0.0282],[165,0.0282],[166,0.0271],[167,0.0274],[168,0.0089],[169,0.0106],[170,0.0193],[171,0.0313],[172,0.0371],[173,0.0416],[174,0.0396],[175,0.0427],[176,0.0451],[177,0.0503],[179,0.0626],[180,0.0677],[181,0.0717],[182,0.0529],[183,0.0011],[184,-0.0086]]
       }
       chest_tan_ctrl.Location.Y {
        type=MovieSceneScriptingFloatChannel
        default=-1.0843
        thinned_from=185
        keys [[0,-2.023],[15,-2.0259],[47,-2.0149],[51,-2.006],[52,-1.9944],[53,-1.9754],[55,-1.9261],[56,-1.8911],[61,-1.675],[62,-1.6437],[63,-1.6168],[64,-1.5833],[67,-1.5003],[71,-1.3815],[73,-1.3197],[74,-1.2936],[76,-1.25],[77,-1.2304],[80,-1.1833],[82,-1.1648],[86,-1.1356],[93,-1.0969],[96,-1.0845],[99,-1.0769],[103,-1.0782],[105,-1.0961],[106,-1.0789],[127,-1.0751],[153,-1.0711],[154,-1.0677],[155,-1.0677],[156,-1.0655],[166,-1.0618],[167,-1.0644],[168,-1.0748],[170,-1.0611],[171,-1.0447],[172,-1.0386],[174,-1.0204],[176,-1.0179],[178,-1.0276],[179,-1.029],[180,-1.0277],[181,-1.0195],[182,-1.0273],[183,-1.1047],[184,-1.0843]]
       }
       chest_tan_ctrl.Location.Z {
        type=MovieSceneScriptingFloatChannel
        default=0.2307
        thinned_from=185
        keys [[0,0.7712],[50,0.7618],[51,0.7553],[52,0.7374],[53,0.71],[54,0.6765],[55,0.6395],[57,0.5563],[58,0.516],[59,0.4794],[60,0.4457],[61,0.4177],[62,0.397],[63,0.3804],[64,0.3626],[65,0.3467],[66,0.3321],[68,0.3051],[69,0.2927],[70,0.2812],[71,0.2708],[72,0.261],[73,0.2524],[74,0.245],[75,0.2387],[76,0.2333],[77,0.2287],[78,0.2248],[79,0.2216],[80,0.2189],[81,0.2169],[88,0.2093],[95,0.2051],[167,0.2039],[170,0.2044],[172,0.2036],[177,0.2035],[178,0.4689],[179,0.2035],[180,0.2041],[181,0.2063],[182,0.2134],[183,0.227],[184,0.2307]]
       }
       chest_tan_ctrl.Rotation.X {
        type=MovieSceneScriptingFloatChannel
        default=7.8037
        thinned_from=185
        keys [[0,11.1257],[15,11.1506],[47,11.0445],[51,10.9747],[52,10.8895],[55,10.448],[56,10.2365],[61,9.0494],[62,8.8989],[63,8.7746],[64,8.6016],[65,8.4632],[67,8.2438],[71,7.7235],[72,7.5731],[73,7.4382],[74,7.328],[77,7.081],[80,6.8955],[86,6.7023],[96,6.4967],[99,6.4589],[102,6.4601],[103,6.4695],[105,6.6025],[106,6.4745],[127,6.4535],[128,6.4336],[133,6.4401],[134,6.4295],[135,6.435],[136,6.4213],[141,6.4191],[142,6.4269],[143,6.4092],[144,6.4177],[145,6.4086],[146,6.4178],[153,6.4252],[154,6.4008],[155,6.4008],[156,6.3854],[166,6.3604],[167,6.3797],[168,6.456],[170,6.3572],[171,6.2386],[172,6.1947],[173,6.1224],[174,6.0632],[176,6.0475],[177,6.0808],[178,6.1],[180,6.1672],[181,6.2408],[182,6.6896],[183,7.7741],[184,7.8037]]
       }
       chest_tan_ctrl.Rotation.Y {
        type=MovieSceneScriptingFloatChannel
        default=0
        thinned_from=185
        keys [[0,0],[79,0],[80,0.0208],[81,0.0395],[83,0.082],[84,0.0925],[85,0.1062],[86,0.134],[87,0.1507],[88,0.1461],[89,0.1378],[90,0.1409],[91,0.1492],[92,0.1554],[93,0.1575],[94,0.162],[95,0.1697],[96,0.1798],[97,0.1878],[98,0.2006],[99,0.2121],[100,0.2259],[101,0.2367],[102,0.2367],[103,0.2233],[104,0.1382],[105,0.0864],[106,0.2151],[107,0.2086],[108,0.2151],[109,0.2147],[110,0.2197],[111,0.2142],[112,0.2062],[113,0.1987],[114,0.1944],[115,0.1911],[116,0.1808],[117,0.1716],[118,0.1704],[119,0.1739],[120,0.1704],[121,0.1837],[122,0.1889],[123,0.1837],[124,0.1957],[127,0.1957],[128,0.2033],[129,0.2033],[130,0.2046],[133,0.2046],[134,0.2264],[135,0.2264],[136,0.2391],[137,0.2399],[138,0.2431],[140,0.2318],[141,0.2296],[142,0.2296],[143,0.2382],[144,0.2382],[145,0.2539],[146,0.2539],[147,0.2628],[148,0.2615],[149,0.2628],[151,0.2628],[152,0.2468],[153,0.2468],[154,0.2291],[155,0.2291],[156,0.2132],[157,0.2132],[158,0.1992],[159,0.1992],[160,0.1978],[161,0.1978],[162,0.2017],[165,0.2017],[166,0.1968],[167,0.1996],[168,0.0619],[169,0.0739],[170,0.1412],[171,0.2288],[172,0.2717],[173,0.3076],[174,0.2976],[175,0.3184],[176,0.3362],[177,0.3734],[178,0.4336],[179,0.4635],[180,0.5001],[181,0.5275],[182,0.3864],[183,0.0249],[184,0]]
       }
       chest_tan_ctrl.Rotation.Z {
        type=MovieSceneScriptingFloatChannel
        default=0
        thinned_from=185
        keys [[0,0.9607],[2,0.9607],[3,0.9424],[6,0.9424],[7,0.9297],[8,0.9064],[9,0.8871],[11,0.8861],[12,0.8781],[13,0.8653],[15,0.8653],[16,0.8701],[17,0.8838],[18,0.8954],[19,0.9116],[20,0.9116],[21,0.9308],[22,0.961],[23,0.9826],[24,0.9826],[25,0.9953],[27,0.9953],[29,0.9374],[30,0.9374],[31,0.9404],[34,0.9404],[35,0.931],[36,0.931],[37,0.9499],[39,0.9499],[40,0.9631],[41,0.9631],[42,0.9456],[43,0.9608],[44,0.9608],[45,0.9795],[46,0.9913],[47,0.9913],[48,0.9748],[49,0.9651],[50,0.9403],[51,0.9216],[52,0.8944],[53,0.8225],[54,0.7729],[55,0.6965],[56,0.6393],[57,0.5965],[58,0.5858],[59,0.5132],[60,0.3815],[61,0.2539],[62,0.205],[63,0.1696],[64,0.2111],[65,0.2307],[66,0.2135],[67,0.1161],[68,0],[74,0],[75,0.0512],[76,0.1437],[77,0.2018],[78,0.2238],[79,0.2523],[80,0.2182],[81,0.2207],[82,0.206],[83,0.195],[84,0.2003],[85,0.1934],[86,0.1861],[87,0.1818],[88,0.1921],[89,0.2098],[90,0.2203],[91,0.2331],[92,0.2466],[93,0.2514],[94,0.2494],[95,0.2385],[96,0.2093],[97,0.1646],[98,0.1023],[99,0.0407],[100,0],[103,0],[104,0.2398],[105,0.3844],[106,0],[111,0],[112,0.0281],[113,0.0594],[114,0.0865],[115,0.132],[116,0.1664],[117,0.1943],[118,0.2174],[119,0.229],[120,0.2174],[121,0.2188],[122,0.2107],[123,0.2188],[124,0.2163],[127,0.2163],[128,0.2181],[129,0.2181],[130,0.2321],[133,0.2321],[134,0.2237],[135,0.2237],[136,0.2197],[137,0.2015],[138,0.2121],[139,0.2259],[140,0.2602],[141,0.2811],[142,0.2811],[143,0.2626],[144,0.2626],[145,0.2408],[146,0.2408],[147,0.2253],[148,0.2215],[149,0.2253],[151,0.2253],[152,0.2353],[153,0.2353],[154,0.2504],[155,0.2504],[156,0.2587],[157,0.2587],[158,0.278],[159,0.278],[160,0.291],[161,0.291],[162,0.2782],[165,0.2782],[166,0.2372],[167,0.1851],[168,0.228],[169,0.3608],[170,0.418],[171,0.7372],[172,0.8241],[173,0.9383],[174,1.0779],[175,1.2094],[176,1.21],[177,1.1875],[178,1.4186],[179,1.0889],[180,1.0421],[181,0.9033],[182,0.3428],[183,0],[184,0]]
       }
       chest_tan_ctrl.Scale.X {type=MovieSceneScriptingFloatChannel, default=1, const [1,185]}
       chest_tan_ctrl.Scale.Y {type=MovieSceneScriptingFloatChannel, default=1, const [1,185]}
       chest_tan_ctrl.Scale.Z {type=MovieSceneScriptingFloatChannel, default=1, const [1,185]}
      ]
     }
     {
      type=MovieSceneControlRigParameterSection
      channels [
       global_ctrl.Location.X {type=MovieSceneScriptingFloatChannel, default=0}
       global_ctrl.Location.Y {type=MovieSceneScriptingFloatChannel, default=0}
       global_ctrl.Location.Z {type=MovieSceneScriptingFloatChannel, default=0}
       global_ctrl.Rotation.X {type=MovieSceneScriptingFloatChannel, default=0}
       global_ctrl.Rotation.Y {type=MovieSceneScriptingFloatChannel, default=0}
       global_ctrl.Rotation.Z {type=MovieSceneScriptingFloatChannel, default=0}
       root_ctrl.Location.X {type=MovieSceneScriptingFloatChannel, default=0}
       root_ctrl.Location.Y {type=MovieSceneScriptingFloatChannel, default=0}
       root_ctrl.Location.Z {type=MovieSceneScriptingFloatChannel, default=0}
       root_ctrl.Rotation.X {type=MovieSceneScriptingFloatChannel, default=0}
       root_ctrl.Rotation.Y {type=MovieSceneScriptingFloatChannel, default=0}
       root_ctrl.Rotation.Z {type=MovieSceneScriptingFloatChannel, default=0}
       root_ctrl.Scale.X {type=MovieSceneScriptingFloatChannel, default=0}
       root_ctrl.Scale.Y {type=MovieSceneScriptingFloatChannel, default=0}
       root_ctrl.Scale.Z {type=MovieSceneScriptingFloatChannel, default=0}
       body_offset_ctrl.Location.X {type=MovieSceneScriptingFloatChannel, default=0}
       body_offset_ctrl.Location.Y {type=MovieSceneScriptingFloatChannel, default=0}
       body_offset_ctrl.Location.Z {type=MovieSceneScriptingFloatChannel, default=0}
       body_offset_ctrl.Rotation.X {type=MovieSceneScriptingFloatChannel, default=0}
       body_offset_ctrl.Rotation.Y {type=MovieSceneScriptingFloatChannel, default=0}
       body_offset_ctrl.Rotation.Z {type=MovieSceneScriptingFloatChannel, default=0}
       body_offset_ctrl.Scale.X {type=MovieSceneScriptingFloatChannel, default=0}
       body_offset_ctrl.Scale.Y {type=MovieSceneScriptingFloatChannel, default=0}
       body_offset_ctrl.Scale.Z {type=MovieSceneScriptingFloatChannel, default=0}
       thigh_l_fk_ctrl.Location.X {type=MovieSceneScriptingFloatChannel, default=0, thinned_from=7, keys [[50,0],[183,0]]}
       thigh_l_fk_ctrl.Location.Y {type=MovieSceneScriptingFloatChannel, default=0, thinned_from=7, keys [[50,0],[183,0]]}
       thigh_l_fk_ctrl.Location.Z {type=MovieSceneScriptingFloatChannel, default=0, thinned_from=7, keys [[50,0],[183,0]]}
       thigh_l_fk_ctrl.Rotation.X {type=MovieSceneScriptingFloatChannel, default=0, keys [[50,-5.9512],[69,0],[98,-0.078],[161,-0.078],[168,-0.0855],[179,-2.6101],[183,0]]}
       thigh_l_fk_ctrl.Rotation.Y {type=MovieSceneScriptingFloatChannel, default=0, keys [[50,-5.1312],[69,0],[98,2.1451],[161,2.1451],[168,2.1642],[179,8.6107],[183,0]]}
       thigh_l_fk_ctrl.Rotation.Z {type=MovieSceneScriptingFloatChannel, default=0, keys [[50,4.8918],[69,5.9382],[98,6.4649],[161,6.4649],[168,6.3968],[179,-16.6272],[183,0]]}
       thigh_l_fk_ctrl.Scale.X {type=MovieSceneScriptingFloatChannel, default=0, const [0,7]}
       thigh_l_fk_ctrl.Scale.Y {type=MovieSceneScriptingFloatChannel, default=0, const [0,7]}
       thigh_l_fk_ctrl.Scale.Z {type=MovieSceneScriptingFloatChannel, default=0, const [0,7]}
       calf_l_fk_ctrl.Location.X {type=MovieSceneScriptingFloatChannel, default=0, thinned_from=7, keys [[50,-0.0002],[184,0]]}
       calf_l_fk_ctrl.Location.Y {type=MovieSceneScriptingFloatChannel, default=0, thinned_from=7, keys [[50,-0.0022],[98,-0.0016],[161,-0.0001],[184,0]]}
       calf_l_fk_ctrl.Location.Z {type=MovieSceneScriptingFloatChannel, default=0, thinned_from=7, keys [[50,-0.0018],[98,-0.0013],[161,-0.0001],[184,0]]}
       calf_l_fk_ctrl.Rotation.X {type=MovieSceneScriptingFloatChannel, default=1.7662, thinned_from=7, keys [[50,4.0717],[58,4.0462],[168,4.0281],[179,4.3916],[184,1.7662]]}
       calf_l_fk_ctrl.Rotation.Y {
        type=MovieSceneScriptingFloatChannel
        default=1.1341
        keys [[50,0.2575],[58,0.2672],[98,0.5214],[161,1.0791],[168,1.0715],[179,-1.4897],[184,1.1341]]
       }
       calf_l_fk_ctrl.Rotation.Z {
        type=MovieSceneScriptingFloatChannel
        default=21.1116
        keys [[50,-7.6111],[58,11.7574],[69,25.4746],[98,-7.5352],[161,-7.3813],[168,-7.2912],[179,23.1722],[184,21.1116]]
       }
       calf_l_fk_ctrl.Scale.X {type=MovieSceneScriptingFloatChannel, default=0, const [0,7]}
       calf_l_fk_ctrl.Scale.Y {type=MovieSceneScriptingFloatChannel, default=0, const [0,7]}
       calf_l_fk_ctrl.Scale.Z {type=MovieSceneScriptingFloatChannel, default=0, const [0,7]}
       foot_l_fk_ctrl.Location.X {type=MovieSceneScriptingFloatChannel, default=0, thinned_from=4, keys [[50,0],[179,0]]}
       foot_l_fk_ctrl.Location.Y {type=MovieSceneScriptingFloatChannel, default=0, thinned_from=4, keys [[50,0],[179,0]]}
       foot_l_fk_ctrl.Location.Z {type=MovieSceneScriptingFloatChannel, default=0, thinned_from=4, keys [[50,-0.0001],[179,0]]}
       foot_l_fk_ctrl.Rotation.X {type=MovieSceneScriptingFloatChannel, default=0, keys [[50,-6.2379],[98,-19.3752],[161,-8.6086],[179,0]]}
       foot_l_fk_ctrl.Rotation.Y {type=MovieSceneScriptingFloatChannel, default=0, keys [[50,-7.8481],[98,-7.1731],[161,-1.7618],[179,0]]}
       foot_l_fk_ctrl.Rotation.Z {type=MovieSceneScriptingFloatChannel, default=0, keys [[50,8.687],[98,10.0347],[161,13.4813],[179,0]]}
       foot_l_fk_ctrl.Scale.X {type=MovieSceneScriptingFloatChannel, default=0, const [0,4]}
       foot_l_fk_ctrl.Scale.Y {type=MovieSceneScriptingFloatChannel, default=0, const [0,4]}
       foot_l_fk_ctrl.Scale.Z {type=MovieSceneScriptingFloatChannel, default=0, const [0,4]}
       ball_l_fk_ctrl.Location.X {type=MovieSceneScriptingFloatChannel, default=0}
       ball_l_fk_ctrl.Location.Y {type=MovieSceneScriptingFloatChannel, default=0}
       ball_l_fk_ctrl.Location.Z {type=MovieSceneScriptingFloatChannel, default=0}
       ball_l_fk_ctrl.Rotation.X {type=MovieSceneScriptingFloatChannel, default=0}
       ball_l_fk_ctrl.Rotation.Y {type=MovieSceneScriptingFloatChannel, default=0}
       ball_l_fk_ctrl.Rotation.Z {type=MovieSceneScriptingFloatChannel, default=0}
       ball_l_fk_ctrl.Scale.X {type=MovieSceneScriptingFloatChannel, default=0}
       ball_l_fk_ctrl.Scale.Y {type=MovieSceneScriptingFloatChannel, default=0}
       ball_l_fk_ctrl.Scale.Z {type=MovieSceneScriptingFloatChannel, default=0}
       foot_l_ik_ctrl.Location.X {type=MovieSceneScriptingFloatChannel, default=0}
       foot_l_ik_ctrl.Location.Y {type=MovieSceneScriptingFloatChannel, default=0}
       foot_l_ik_ctrl.Location.Z {type=MovieSceneScriptingFloatChannel, default=0}
       foot_l_ik_ctrl.Rotation.X {type=MovieSceneScriptingFloatChannel, default=0}
       foot_l_ik_ctrl.Rotation.Y {type=MovieSceneScriptingFloatChannel, default=0}
       foot_l_ik_ctrl.Rotation.Z {type=MovieSceneScriptingFloatChannel, default=0}
       foot_l_ik_ctrl.Scale.X {type=MovieSceneScriptingFloatChannel, default=0}
       foot_l_ik_ctrl.Scale.Y {type=MovieSceneScriptingFloatChannel, default=0}
       foot_l_ik_ctrl.Scale.Z {type=MovieSceneScriptingFloatChannel, default=0}
       ball_l_ik_ctrl.Location.X {type=MovieSceneScriptingFloatChannel, default=0}
       ball_l_ik_ctrl.Location.Y {type=MovieSceneScriptingFloatChannel, default=0}
       ball_l_ik_ctrl.Location.Z {type=MovieSceneScriptingFloatChannel, default=0}
       ball_l_ik_ctrl.Rotation.X {type=MovieSceneScriptingFloatChannel, default=0}
       ball_l_ik_ctrl.Rotation.Y {type=MovieSceneScriptingFloatChannel, default=0}
       ball_l_ik_ctrl.Rotation.Z {type=MovieSceneScriptingFloatChannel, default=0}
       ball_l_ik_ctrl.Scale.X {type=MovieSceneScriptingFloatChannel, default=0}
       ball_l_ik_ctrl.Scale.Y {type=MovieSceneScriptingFloatChannel, default=0}
       ball_l_ik_ctrl.Scale.Z {type=MovieSceneScriptingFloatChannel, default=0}
       foot_bk1_l_ctrl.Location.X {type=MovieSceneScriptingFloatChannel, default=0}
       foot_bk1_l_ctrl.Location.Y {type=MovieSceneScriptingFloatChannel, default=0}
       foot_bk1_l_ctrl.Location.Z {type=MovieSceneScriptingFloatChannel, default=0}
       foot_bk1_l_ctrl.Rotation.X {type=MovieSceneScriptingFloatChannel, default=0}
       foot_bk1_l_ctrl.Rotation.Y {type=MovieSceneScriptingFloatChannel, default=0}
       foot_bk1_l_ctrl.Rotation.Z {type=MovieSceneScriptingFloatChannel, default=0}
       foot_bk1_l_ctrl.Scale.X {type=MovieSceneScriptingFloatChannel, default=0}
       foot_bk1_l_ctrl.Scale.Y {type=MovieSceneScriptingFloatChannel, default=0}
       foot_bk1_l_ctrl.Scale.Z {type=MovieSceneScriptingFloatChannel, default=0}
       foot_roll_l_ctrl.Location.X {type=MovieSceneScriptingFloatChannel, default=0}
       foot_roll_l_ctrl.Location.Y {type=MovieSceneScriptingFloatChannel, default=0}
       foot_roll_l_ctrl.Location.Z {type=MovieSceneScriptingFloatChannel, default=0}
       foot_roll_l_ctrl.Rotation.X {type=MovieSceneScriptingFloatChannel, default=0}
       foot_roll_l_ctrl.Rotation.Y {type=MovieSceneScriptingFloatChannel, default=0}
       foot_roll_l_ctrl.Rotation.Z {type=MovieSceneScriptingFloatChannel, default=0}
       foot_roll_l_ctrl.Scale.X {type=MovieSceneScriptingFloatChannel, default=0}
       foot_roll_l_ctrl.Scale.Y {type=MovieSceneScriptingFloatChannel, default=0}
       foot_roll_l_ctrl.Scale.Z {type=MovieSceneScriptingFloatChannel, default=0}
       roll_blend_l_ctrl {type=MovieSceneScriptingFloatChannel, default=0}
       tip_l_ctrl.Location.X {type=MovieSceneScriptingFloatChannel, default=0}
       tip_l_ctrl.Location.Y {type=MovieSceneScriptingFloatChannel, default=0}
       tip_l_ctrl.Location.Z {type=MovieSceneScriptingFloatChannel, default=0}
       tip_l_ctrl.Rotation.X {type=MovieSceneScriptingFloatChannel, default=0}
       tip_l_ctrl.Rotation.Y {type=MovieSceneScriptingFloatChannel, default=0}
       tip_l_ctrl.Rotation.Z {type=MovieSceneScriptingFloatChannel, default=0}
       tip_l_ctrl.Scale.X {type=MovieSceneScriptingFloatChannel, default=0}
       tip_l_ctrl.Scale.Y {type=MovieSceneScriptingFloatChannel, default=0}
       tip_l_ctrl.Scale.Z {type=MovieSceneScriptingFloatChannel, default=0}
       heel_l_ctrl.Location.X {type=MovieSceneScriptingFloatChannel, default=0}
       heel_l_ctrl.Location.Y {type=MovieSceneScriptingFloatChannel, default=0}
       heel_l_ctrl.Location.Z {type=MovieSceneScriptingFloatChannel, default=0}
       heel_l_ctrl.Rotation.X {type=MovieSceneScriptingFloatChannel, default=0}
       heel_l_ctrl.Rotation.Y {type=MovieSceneScriptingFloatChannel, default=0}
       heel_l_ctrl.Rotation.Z {type=MovieSceneScriptingFloatChannel, default=0}
       heel_l_ctrl.Scale.X {type=MovieSceneScriptingFloatChannel, default=0}
       heel_l_ctrl.Scale.Y {type=MovieSceneScriptingFloatChannel, default=0}
       heel_l_ctrl.Scale.Z {type=MovieSceneScriptingFloatChannel, default=0}
       leg_l_softness {type=MovieSceneScriptingFloatChannel, default=0}
       leg_l_pv_ik_ctrl.Location.X {type=MovieSceneScriptingFloatChannel, default=0}
       leg_l_pv_ik_ctrl.Location.Y {type=MovieSceneScriptingFloatChannel, default=0}
       leg_l_pv_ik_ctrl.Location.Z {type=MovieSceneScriptingFloatChannel, default=0}
       leg_l_pv_ik_ctrl.Rotation.X {type=MovieSceneScriptingFloatChannel, default=0}
       leg_l_pv_ik_ctrl.Rotation.Y {type=MovieSceneScriptingFloatChannel, default=0}
       leg_l_pv_ik_ctrl.Rotation.Z {type=MovieSceneScriptingFloatChannel, default=0}
       leg_l_pv_ik_ctrl.Scale.X {type=MovieSceneScriptingFloatChannel, default=0}
       leg_l_pv_ik_ctrl.Scale.Y {type=MovieSceneScriptingFloatChannel, default=0}
       leg_l_pv_ik_ctrl.Scale.Z {type=MovieSceneScriptingFloatChannel, default=0}
       thigh_r_fk_ctrl.Location.X {type=MovieSceneScriptingFloatChannel, default=0, thinned_from=10, keys [[42,0.0002],[59,0],[183,0]]}
       thigh_r_fk_ctrl.Location.Y {type=MovieSceneScriptingFloatChannel, default=0, thinned_from=10, keys [[42,0],[183,0]]}
       thigh_r_fk_ctrl.Location.Z {type=MovieSceneScriptingFloatChannel, default=0, thinned_from=10, keys [[42,0],[183,0]]}
       thigh_r_fk_ctrl.Rotation.X {
        type=MovieSceneScriptingFloatChannel
        default=0
        keys [[42,0.8733],[45,0.8656],[50,0.8621],[58,-0.7421],[59,-0.7464],[98,-0.8566],[161,-0.8566],[168,-0.855],[179,-0.1007],[183,0]]
       }
       thigh_r_fk_ctrl.Rotation.Y {
        type=MovieSceneScriptingFloatChannel
        default=0
        keys [[42,3.3147],[45,3.2856],[50,3.2865],[58,0.3046],[59,0.3047],[98,0.3516],[161,0.3516],[168,0.3584],[179,2.651],[183,0]]
       }
       thigh_r_fk_ctrl.Rotation.Z {
        type=MovieSceneScriptingFloatChannel
        default=0
        keys [[42,7.4383],[45,7.3731],[50,7.4255],[59,16.9534],[98,21.3581],[161,21.3581],[168,21.2681],[179,-9.1785],[183,0]]
       }
       thigh_r_fk_ctrl.Scale.X {type=MovieSceneScriptingFloatChannel, default=0, const [0,10]}
       thigh_r_fk_ctrl.Scale.Y {type=MovieSceneScriptingFloatChannel, default=0, const [0,10]}
       thigh_r_fk_ctrl.Scale.Z {type=MovieSceneScriptingFloatChannel, default=0, const [0,10]}
       calf_r_fk_ctrl.Location.X {type=MovieSceneScriptingFloatChannel, default=0, thinned_from=10, keys [[42,-0.0002],[59,0],[183,0]]}
       calf_r_fk_ctrl.Location.Y {type=MovieSceneScriptingFloatChannel, default=0, thinned_from=10, keys [[42,0.0004],[50,0.0004],[59,0],[183,0]]}
       calf_r_fk_ctrl.Location.Z {type=MovieSceneScriptingFloatChannel, default=0, thinned_from=10, keys [[42,0],[183,0]]}
       calf_r_fk_ctrl.Rotation.X {
        type=MovieSceneScriptingFloatChannel
        default=0
        keys [[42,7.8257],[45,7.757],[50,7.7436],[58,-2.8376],[59,-2.8564],[98,-3.3415],[161,-3.3415],[168,-3.3308],[179,0.2777],[183,0]]
       }
       calf_r_fk_ctrl.Rotation.Y {
        type=MovieSceneScriptingFloatChannel
        default=0
        keys [[42,0.2127],[45,0.2108],[50,0.2064],[58,-1.0799],[59,-1.0871],[98,-1.2717],[161,-1.2717],[168,-1.2697],[179,-0.3031],[183,0]]
       }
       calf_r_fk_ctrl.Rotation.Z {
        type=MovieSceneScriptingFloatChannel
        default=0
        keys [[42,-16.4773],[45,-16.3328],[50,-16.4004],[59,-5.5156],[69,-2.9744],[98,-28.2013],[161,-28.2013],[168,-28.0929],[179,8.5509],[183,0]]
       }
       calf_r_fk_ctrl.Scale.X {type=MovieSceneScriptingFloatChannel, default=0, const [0,10]}
       calf_r_fk_ctrl.Scale.Y {type=MovieSceneScriptingFloatChannel, default=0, const [0,10]}
       calf_r_fk_ctrl.Scale.Z {type=MovieSceneScriptingFloatChannel, default=0, const [0,10]}
       foot_r_fk_ctrl.Location.X {type=MovieSceneScriptingFloatChannel, default=0, thinned_from=8, keys [[42,-0.0001],[183,0]]}
       foot_r_fk_ctrl.Location.Y {type=MovieSceneScriptingFloatChannel, default=0, thinned_from=8, keys [[42,0],[183,0]]}
       foot_r_fk_ctrl.Location.Z {type=MovieSceneScriptingFloatChannel, default=0, thinned_from=8, keys [[42,0],[183,0]]}
       foot_r_fk_ctrl.Rotation.X {
        type=MovieSceneScriptingFloatChannel
        default=0
        thinned_from=8
        keys [[42,-13.7794],[45,-13.6586],[50,-13.6814],[58,-8.0632],[98,-9.4952],[183,-9.4952]]
       }
       foot_r_fk_ctrl.Rotation.Y {
        type=MovieSceneScriptingFloatChannel
        default=0
        thinned_from=8
        keys [[42,2.4765],[45,2.4547],[50,2.4309],[58,-5.4955],[98,-6.4715],[183,-6.4715]]
       }
       foot_r_fk_ctrl.Rotation.Z {
        type=MovieSceneScriptingFloatChannel
        default=0
        thinned_from=8
        keys [[42,5.1448],[45,5.0997],[50,5.1603],[58,21.446],[98,25.2549],[183,25.2549]]
       }
       foot_r_fk_ctrl.Scale.X {type=MovieSceneScriptingFloatChannel, default=0, const [0,8]}
       foot_r_fk_ctrl.Scale.Y {type=MovieSceneScriptingFloatChannel, default=0, const [0,8]}
       foot_r_fk_ctrl.Scale.Z {type=MovieSceneScriptingFloatChannel, default=0, const [0,8]}
       ball_r_fk_ctrl.Location.X {type=MovieSceneScriptingFloatChannel, default=0}
       ball_r_fk_ctrl.Location.Y {type=MovieSceneScriptingFloatChannel, default=0}
       ball_r_fk_ctrl.Location.Z {type=MovieSceneScriptingFloatChannel, default=0}
       ball_r_fk_ctrl.Rotation.X {type=MovieSceneScriptingFloatChannel, default=0}
       ball_r_fk_ctrl.Rotation.Y {type=MovieSceneScriptingFloatChannel, default=0}
       ball_r_fk_ctrl.Rotation.Z {type=MovieSceneScriptingFloatChannel, default=0}
       ball_r_fk_ctrl.Scale.X {type=MovieSceneScriptingFloatChannel, default=0}
       ball_r_fk_ctrl.Scale.Y {type=MovieSceneScriptingFloatChannel, default=0}
       ball_r_fk_ctrl.Scale.Z {type=MovieSceneScriptingFloatChannel, default=0}
       foot_r_ik_ctrl.Location.X {type=MovieSceneScriptingFloatChannel, default=0}
       foot_r_ik_ctrl.Location.Y {type=MovieSceneScriptingFloatChannel, default=0}
       foot_r_ik_ctrl.Location.Z {type=MovieSceneScriptingFloatChannel, default=0}
       foot_r_ik_ctrl.Rotation.X {type=MovieSceneScriptingFloatChannel, default=0}
       foot_r_ik_ctrl.Rotation.Y {type=MovieSceneScriptingFloatChannel, default=0}
       foot_r_ik_ctrl.Rotation.Z {type=MovieSceneScriptingFloatChannel, default=0}
       foot_r_ik_ctrl.Scale.X {type=MovieSceneScriptingFloatChannel, default=0}
       foot_r_ik_ctrl.Scale.Y {type=MovieSceneScriptingFloatChannel, default=0}
       foot_r_ik_ctrl.Scale.Z {type=MovieSceneScriptingFloatChannel, default=0}
       ball_r_ik_ctrl.Location.X {type=MovieSceneScriptingFloatChannel, default=0}
       ball_r_ik_ctrl.Location.Y {type=MovieSceneScriptingFloatChannel, default=0}
       ball_r_ik_ctrl.Location.Z {type=MovieSceneScriptingFloatChannel, default=0}
       ball_r_ik_ctrl.Rotation.X {type=MovieSceneScriptingFloatChannel, default=0}
       ball_r_ik_ctrl.Rotation.Y {type=MovieSceneScriptingFloatChannel, default=0}
       ball_r_ik_ctrl.Rotation.Z {type=MovieSceneScriptingFloatChannel, default=0}
       ball_r_ik_ctrl.Scale.X {type=MovieSceneScriptingFloatChannel, default=0}
       ball_r_ik_ctrl.Scale.Y {type=MovieSceneScriptingFloatChannel, default=0}
       ball_r_ik_ctrl.Scale.Z {type=MovieSceneScriptingFloatChannel, default=0}
       foot_bk1_r_ctrl.Location.X {type=MovieSceneScriptingFloatChannel, default=0}
       foot_bk1_r_ctrl.Location.Y {type=MovieSceneScriptingFloatChannel, default=0}
       foot_bk1_r_ctrl.Location.Z {type=MovieSceneScriptingFloatChannel, default=0}
       foot_bk1_r_ctrl.Rotation.X {type=MovieSceneScriptingFloatChannel, default=0}
       foot_bk1_r_ctrl.Rotation.Y {type=MovieSceneScriptingFloatChannel, default=0}
       foot_bk1_r_ctrl.Rotation.Z {type=MovieSceneScriptingFloatChannel, default=0}
       foot_bk1_r_ctrl.Scale.X {type=MovieSceneScriptingFloatChannel, default=0}
       foot_bk1_r_ctrl.Scale.Y {type=MovieSceneScriptingFloatChannel, default=0}
       foot_bk1_r_ctrl.Scale.Z {type=MovieSceneScriptingFloatChannel, default=0}
       foot_roll_r_ctrl.Location.X {type=MovieSceneScriptingFloatChannel, default=0}
       foot_roll_r_ctrl.Location.Y {type=MovieSceneScriptingFloatChannel, default=0}
       foot_roll_r_ctrl.Location.Z {type=MovieSceneScriptingFloatChannel, default=0}
       foot_roll_r_ctrl.Rotation.X {type=MovieSceneScriptingFloatChannel, default=0}
       foot_roll_r_ctrl.Rotation.Y {type=MovieSceneScriptingFloatChannel, default=0}
       foot_roll_r_ctrl.Rotation.Z {type=MovieSceneScriptingFloatChannel, default=0}
       foot_roll_r_ctrl.Scale.X {type=MovieSceneScriptingFloatChannel, default=0}
       foot_roll_r_ctrl.Scale.Y {type=MovieSceneScriptingFloatChannel, default=0}
       foot_roll_r_ctrl.Scale.Z {type=MovieSceneScriptingFloatChannel, default=0}
       roll_blend_r_ctrl {type=MovieSceneScriptingFloatChannel, default=0}
       tip_r_ctrl.Location.X {type=MovieSceneScriptingFloatChannel, default=0}
       tip_r_ctrl.Location.Y {type=MovieSceneScriptingFloatChannel, default=0}
       tip_r_ctrl.Location.Z {type=MovieSceneScriptingFloatChannel, default=0}
       tip_r_ctrl.Rotation.X {type=MovieSceneScriptingFloatChannel, default=0}
       tip_r_ctrl.Rotation.Y {type=MovieSceneScriptingFloatChannel, default=0}
       tip_r_ctrl.Rotation.Z {type=MovieSceneScriptingFloatChannel, default=0}
       tip_r_ctrl.Scale.X {type=MovieSceneScriptingFloatChannel, default=0}
       tip_r_ctrl.Scale.Y {type=MovieSceneScriptingFloatChannel, default=0}
       tip_r_ctrl.Scale.Z {type=MovieSceneScriptingFloatChannel, default=0}
       heel_r_ctrl.Location.X {type=MovieSceneScriptingFloatChannel, default=0}
       heel_r_ctrl.Location.Y {type=MovieSceneScriptingFloatChannel, default=0}
       heel_r_ctrl.Location.Z {type=MovieSceneScriptingFloatChannel, default=0}
       heel_r_ctrl.Rotation.X {type=MovieSceneScriptingFloatChannel, default=0}
       heel_r_ctrl.Rotation.Y {type=MovieSceneScriptingFloatChannel, default=0}
       heel_r_ctrl.Rotation.Z {type=MovieSceneScriptingFloatChannel, default=0}
       heel_r_ctrl.Scale.X {type=MovieSceneScriptingFloatChannel, default=0}
       heel_r_ctrl.Scale.Y {type=MovieSceneScriptingFloatChannel, default=0}
       heel_r_ctrl.Scale.Z {type=MovieSceneScriptingFloatChannel, default=0}
       leg_r_softness {type=MovieSceneScriptingFloatChannel, default=0}
       leg_r_pv_ik_ctrl.Location.X {type=MovieSceneScriptingFloatChannel, default=0}
       leg_r_pv_ik_ctrl.Location.Y {type=MovieSceneScriptingFloatChannel, default=0}
       leg_r_pv_ik_ctrl.Location.Z {type=MovieSceneScriptingFloatChannel, default=0}
       leg_r_pv_ik_ctrl.Rotation.X {type=MovieSceneScriptingFloatChannel, default=0}
       leg_r_pv_ik_ctrl.Rotation.Y {type=MovieSceneScriptingFloatChannel, default=0}
       leg_r_pv_ik_ctrl.Rotation.Z {type=MovieSceneScriptingFloatChannel, default=0}
       leg_r_pv_ik_ctrl.Scale.X {type=MovieSceneScriptingFloatChannel, default=0}
       leg_r_pv_ik_ctrl.Scale.Y {type=MovieSceneScriptingFloatChannel, default=0}
       leg_r_pv_ik_ctrl.Scale.Z {type=MovieSceneScriptingFloatChannel, default=0}
       upperarm_l_fk_ctrl.Location.X {type=MovieSceneScriptingFloatChannel, default=0, const [0,7]}
       upperarm_l_fk_ctrl.Location.Y {type=MovieSceneScriptingFloatChannel, default=0, const [0,7]}
       upperarm_l_fk_ctrl.Location.Z {type=MovieSceneScriptingFloatChannel, default=0, const [0,7]}
       upperarm_l_fk_ctrl.Rotation.X {type=MovieSceneScriptingFloatChannel, default=0, keys [[66,0],[94,-2.6592],[164,-2.6592],[174,3.3945],[175,3.397],[178,3.3876],[183,0]]}
       upperarm_l_fk_ctrl.Rotation.Y {type=MovieSceneScriptingFloatChannel, default=0, keys [[66,0],[94,-0.0006],[164,-0.0006],[174,-13.068],[175,-13.0733],[178,-13.037],[183,0]]}
       upperarm_l_fk_ctrl.Rotation.Z {type=MovieSceneScriptingFloatChannel, default=0, keys [[66,0],[94,-0.0004],[164,-0.0004],[174,7.0908],[175,7.0937],[178,7.074],[183,0]]}
       upperarm_l_fk_ctrl.Scale.X {type=MovieSceneScriptingFloatChannel, default=0, const [0,7]}
       upperarm_l_fk_ctrl.Scale.Y {type=MovieSceneScriptingFloatChannel, default=0, const [0,7]}
       upperarm_l_fk_ctrl.Scale.Z {type=MovieSceneScriptingFloatChannel, default=0, const [0,7]}
       lowerarm_l_fk_ctrl.Location.X {type=MovieSceneScriptingFloatChannel, default=0, const [0,7]}
       lowerarm_l_fk_ctrl.Location.Y {type=MovieSceneScriptingFloatChannel, default=0, const [0,7]}
       lowerarm_l_fk_ctrl.Location.Z {type=MovieSceneScriptingFloatChannel, default=0, const [0,7]}
       lowerarm_l_fk_ctrl.Rotation.X {type=MovieSceneScriptingFloatChannel, default=0, thinned_from=7, keys [[66,0],[164,0],[174,1.6478],[175,1.6485],[178,1.6439],[183,0]]}
       lowerarm_l_fk_ctrl.Rotation.Y {type=MovieSceneScriptingFloatChannel, default=0, thinned_from=7, keys [[66,0],[164,0],[174,-0.1927],[175,-0.1928],[178,-0.1922],[183,0]]}
       lowerarm_l_fk_ctrl.Rotation.Z {type=MovieSceneScriptingFloatChannel, default=0, keys [[66,0],[94,4.296],[164,4.2583],[174,-6.8011],[175,-6.8056],[178,-6.7867],[183,0]]}
       lowerarm_l_fk_ctrl.Scale.X {type=MovieSceneScriptingFloatChannel, default=0, const [0,7]}
       lowerarm_l_fk_ctrl.Scale.Y {type=MovieSceneScriptingFloatChannel, default=0, const [0,7]}
       lowerarm_l_fk_ctrl.Scale.Z {type=MovieSceneScriptingFloatChannel, default=0, const [0,7]}
       hand_l_fk_ctrl.Location.X {type=MovieSceneScriptingFloatChannel, default=0, const [0,4]}
       hand_l_fk_ctrl.Location.Y {type=MovieSceneScriptingFloatChannel, default=0, const [0,4]}
       hand_l_fk_ctrl.Location.Z {type=MovieSceneScriptingFloatChannel, default=0, const [0,4]}
       hand_l_fk_ctrl.Rotation.X {type=MovieSceneScriptingFloatChannel, default=0, const [0,4]}
       hand_l_fk_ctrl.Rotation.Y {type=MovieSceneScriptingFloatChannel, default=0, const [0,4]}
       hand_l_fk_ctrl.Rotation.Z {type=MovieSceneScriptingFloatChannel, default=0, const [0,4]}
       hand_l_fk_ctrl.Scale.X {type=MovieSceneScriptingFloatChannel, default=0, const [0,4]}
       hand_l_fk_ctrl.Scale.Y {type=MovieSceneScriptingFloatChannel, default=0, const [0,4]}
       hand_l_fk_ctrl.Scale.Z {type=MovieSceneScriptingFloatChannel, default=0, const [0,4]}
       thumb_01_l_ctrl.Location.X {
        type=MovieSceneScriptingFloatChannel
        default=0
        thinned_from=8
        keys [[42,-0.0001],[53,0.0001],[91,-0.0001],[165,-0.0001],[172,0.0001],[182,-0.0001]]
       }
       thumb_01_l_ctrl.Location.Y {type=MovieSceneScriptingFloatChannel, default=0, thinned_from=8, keys [[42,0.0001],[53,0.0004],[72,0],[182,0.0001]]}
       thumb_01_l_ctrl.Location.Z {type=MovieSceneScriptingFloatChannel, default=0, thinned_from=8, keys [[42,0.0001],[53,0.0003],[72,0],[182,0.0001]]}
       thumb_01_l_ctrl.Rotation.X {
        type=MovieSceneScriptingFloatChannel
        default=0
        keys [[42,-7.1176],[53,23.1504],[72,0],[91,-7.1176],[165,-7.1176],[172,1.5179],[177,-7.1176],[182,-7.1176]]
       }
       thumb_01_l_ctrl.Rotation.Y {
        type=MovieSceneScriptingFloatChannel
        default=0
        keys [[42,-3.2346],[53,-7.803],[72,0],[91,-3.2346],[165,-3.2346],[172,-12.349],[177,-3.2346],[182,-3.2346]]
       }
       thumb_01_l_ctrl.Rotation.Z {
        type=MovieSceneScriptingFloatChannel
        default=0
        keys [[42,9.2367],[53,4.8762],[72,0],[91,9.2367],[165,9.2367],[172,20.1131],[177,9.2367],[182,9.2367]]
       }
       thumb_01_l_ctrl.Scale.X {type=MovieSceneScriptingFloatChannel, default=0, const [0,8]}
       thumb_01_l_ctrl.Scale.Y {type=MovieSceneScriptingFloatChannel, default=0, const [0,8]}
       thumb_01_l_ctrl.Scale.Z {type=MovieSceneScriptingFloatChannel, default=0, const [0,8]}
       thumb_02_l_ctrl.Location.X {type=MovieSceneScriptingFloatChannel, default=0, thinned_from=8, keys [[42,0.0002],[53,0.0004],[72,0],[91,0.0002],[182,0.0002]]}
       thumb_02_l_ctrl.Location.Y {type=MovieSceneScriptingFloatChannel, default=0, thinned_from=8, keys [[42,0],[53,0.0002],[182,0]]}
       thumb_02_l_ctrl.Location.Z {type=MovieSceneScriptingFloatChannel, default=0, thinned_from=8, keys [[42,0],[53,0.0005],[72,0],[182,0]]}
       thumb_02_l_ctrl.Rotation.X {
        type=MovieSceneScriptingFloatChannel
        default=0
        keys [[42,2.9049],[53,3.6262],[72,0],[91,2.9049],[165,2.9049],[172,2.3583],[177,2.9049],[182,2.9049]]
       }
       thumb_02_l_ctrl.Rotation.Y {
        type=MovieSceneScriptingFloatChannel
        default=0
        keys [[42,-8.9743],[53,-7.5918],[72,0],[91,-8.9743],[165,-8.9743],[172,-7.8882],[177,-8.9743],[182,-8.9743]]
       }
       thumb_02_l_ctrl.Rotation.Z {
        type=MovieSceneScriptingFloatChannel
        default=0
        keys [[42,-37.5302],[53,-6.9118],[72,0],[91,-37.5302],[165,-37.5302],[172,-29.411],[177,-37.5302],[182,-37.5302]]
       }
       thumb_02_l_ctrl.Scale.X {type=MovieSceneScriptingFloatChannel, default=0, const [0,8]}
       thumb_02_l_ctrl.Scale.Y {type=MovieSceneScriptingFloatChannel, default=0, const [0,8]}
       thumb_02_l_ctrl.Scale.Z {type=MovieSceneScriptingFloatChannel, default=0, const [0,8]}
       thumb_03_l_ctrl.Location.X {type=MovieSceneScriptingFloatChannel, default=0, thinned_from=8, keys [[42,0],[53,0.0001],[182,0]]}
       thumb_03_l_ctrl.Location.Y {type=MovieSceneScriptingFloatChannel, default=0, thinned_from=8, keys [[42,0],[182,0]]}
       thumb_03_l_ctrl.Location.Z {type=MovieSceneScriptingFloatChannel, default=0, thinned_from=8, keys [[42,0],[53,0.0002],[91,0],[182,0]]}
       thumb_03_l_ctrl.Rotation.X {
        type=MovieSceneScriptingFloatChannel
        default=0
        keys [[42,-0.3219],[53,-0.0533],[72,0],[91,-0.3219],[165,-0.3219],[172,-2.0661],[177,-0.3219],[182,-0.3219]]
       }
       thumb_03_l_ctrl.Rotation.Y {
        type=MovieSceneScriptingFloatChannel
        default=0
        keys [[42,-9.2484],[53,-9.3245],[72,0],[91,-9.2484],[165,-9.2484],[172,-8.8987],[177,-9.2484],[182,-9.2484]]
       }
       thumb_03_l_ctrl.Rotation.Z {
        type=MovieSceneScriptingFloatChannel
        default=0
        keys [[42,-77.6421],[53,-45.0673],[72,0],[91,-77.6421],[165,-77.6421],[172,-23.7829],[177,-77.6421],[182,-77.6421]]
       }
       thumb_03_l_ctrl.Scale.X {type=MovieSceneScriptingFloatChannel, default=0, const [0,8]}
       thumb_03_l_ctrl.Scale.Y {type=MovieSceneScriptingFloatChannel, default=0, const [0,8]}
       thumb_03_l_ctrl.Scale.Z {type=MovieSceneScriptingFloatChannel, default=0, const [0,8]}
       index_metacarpal_l_ctrl.Location.X {
        type=MovieSceneScriptingFloatChannel
        default=0
        thinned_from=8
        keys [[42,0.0002],[53,0.0004],[72,0],[91,0.0002],[165,0.0002],[172,0.0001],[182,0.0002]]
       }
       index_metacarpal_l_ctrl.Location.Y {
        type=MovieSceneScriptingFloatChannel
        default=0
        thinned_from=8
        keys [[42,0.0001],[53,0.0002],[72,0],[165,0.0001],[172,-0.0001],[177,0.0001],[182,0.0001]]
       }
       index_metacarpal_l_ctrl.Location.Z {type=MovieSceneScriptingFloatChannel, default=0, thinned_from=8, keys [[42,0],[53,-0.0003],[72,0],[182,0]]}
       index_metacarpal_l_ctrl.Rotation.X {type=MovieSceneScriptingFloatChannel, default=0, thinned_from=8, keys [[42,-0.0061],[53,-0.0061],[72,0],[91,-0.0061],[182,-0.0061]]}
       index_metacarpal_l_ctrl.Rotation.Y {
        type=MovieSceneScriptingFloatChannel
        default=0
        keys [[42,4.5735],[53,3.2423],[72,0],[91,4.5735],[165,4.5735],[172,0.47],[177,4.5735],[182,4.5735]]
       }
       index_metacarpal_l_ctrl.Rotation.Z {type=MovieSceneScriptingFloatChannel, default=0, thinned_from=8, keys [[42,0.7433],[53,0.7433],[72,0],[91,0.7433],[182,0.7433]]}
       index_metacarpal_l_ctrl.Scale.X {type=MovieSceneScriptingFloatChannel, default=0, const [0,8]}
       index_metacarpal_l_ctrl.Scale.Y {type=MovieSceneScriptingFloatChannel, default=0, const [0,8]}
       index_metacarpal_l_ctrl.Scale.Z {type=MovieSceneScriptingFloatChannel, default=0, const [0,8]}
       index_01_l_ctrl.Location.X {type=MovieSceneScriptingFloatChannel, default=0, thinned_from=8, keys [[42,0.0001],[53,0.001],[72,0],[172,0.0002],[182,0.0001]]}
       index_01_l_ctrl.Location.Y {type=MovieSceneScriptingFloatChannel, default=0, thinned_from=8, keys [[42,0],[53,0.0008],[72,0],[182,0]]}
       index_01_l_ctrl.Location.Z {type=MovieSceneScriptingFloatChannel, default=0, thinned_from=8, keys [[42,0],[53,-0.0001],[182,0]]}
       index_01_l_ctrl.Rotation.X {
        type=MovieSceneScriptingFloatChannel
        default=0
        keys [[42,3.5009],[53,-6.199],[72,0],[91,3.5009],[165,3.5009],[172,0.7147],[177,3.5009],[182,3.5009]]
       }
       index_01_l_ctrl.Rotation.Y {
        type=MovieSceneScriptingFloatChannel
        default=0
        keys [[42,2.4662],[53,10.8849],[72,0],[91,2.4662],[165,2.4662],[172,4.6318],[177,2.4662],[182,2.4662]]
       }
       index_01_l_ctrl.Rotation.Z {
        type=MovieSceneScriptingFloatChannel
        default=0
        keys [[42,10.1165],[53,46.0561],[72,0],[91,10.1165],[165,10.1165],[172,52.0191],[177,10.1165],[182,10.1165]]
       }
       index_01_l_ctrl.Scale.X {type=MovieSceneScriptingFloatChannel, default=0, const [0,8]}
       index_01_l_ctrl.Scale.Y {type=MovieSceneScriptingFloatChannel, default=0, const [0,8]}
       index_01_l_ctrl.Scale.Z {type=MovieSceneScriptingFloatChannel, default=0, const [0,8]}
       index_02_l_ctrl.Location.X {type=MovieSceneScriptingFloatChannel, default=0, thinned_from=8, keys [[42,-0.0001],[172,0],[182,-0.0001]]}
       index_02_l_ctrl.Location.Y {
        type=MovieSceneScriptingFloatChannel
        default=0
        thinned_from=8
        keys [[42,-0.0001],[53,0.0001],[91,-0.0001],[165,-0.0001],[172,0.0001],[182,-0.0001]]
       }
       index_02_l_ctrl.Location.Z {type=MovieSceneScriptingFloatChannel, default=0, thinned_from=8, keys [[42,-0.0001],[182,-0.0001]]}
       index_02_l_ctrl.Rotation.X {
        type=MovieSceneScriptingFloatChannel
        default=0
        keys [[42,6.8809],[53,7.1315],[72,0],[91,6.8809],[165,6.8809],[172,6.5153],[177,6.8809],[182,6.8809]]
       }
       index_02_l_ctrl.Rotation.Y {
        type=MovieSceneScriptingFloatChannel
        default=0
        keys [[42,-6.5166],[53,-5.7424],[72,0],[91,-6.5166],[165,-6.5166],[172,-6.1572],[177,-6.5166],[182,-6.5166]]
       }
       index_02_l_ctrl.Rotation.Z {
        type=MovieSceneScriptingFloatChannel
        default=0
        keys [[42,-47.3718],[53,34.9861],[72,0],[91,-47.3718],[165,-47.3718],[172,26.8556],[177,-47.3718],[182,-47.3718]]
       }
       index_02_l_ctrl.Scale.X {type=MovieSceneScriptingFloatChannel, default=0, const [0,8]}
       index_02_l_ctrl.Scale.Y {type=MovieSceneScriptingFloatChannel, default=0, const [0,8]}
       index_02_l_ctrl.Scale.Z {type=MovieSceneScriptingFloatChannel, default=0, const [0,8]}
       index_03_l_ctrl.Location.X {type=MovieSceneScriptingFloatChannel, default=0, thinned_from=8, keys [[42,0.0001],[53,0.0001],[72,0],[182,0.0001]]}
       index_03_l_ctrl.Location.Y {type=MovieSceneScriptingFloatChannel, default=0, thinned_from=8, keys [[42,-0.0002],[53,-0.0003],[72,0],[91,-0.0002],[182,-0.0002]]}
       index_03_l_ctrl.Location.Z {type=MovieSceneScriptingFloatChannel, default=0, thinned_from=8, keys [[42,0],[165,0],[172,0.0001],[182,0]]}
       index_03_l_ctrl.Rotation.X {
        type=MovieSceneScriptingFloatChannel
        default=0
        keys [[42,0.9134],[53,1.2373],[72,0],[91,0.9134],[165,0.9134],[172,1.4541],[177,0.9134],[182,0.9134]]
       }
       index_03_l_ctrl.Rotation.Y {
        type=MovieSceneScriptingFloatChannel
        default=0
        keys [[42,-2.6348],[53,-2.3218],[72,0],[91,-2.6348],[165,-2.6348],[172,-2.8077],[177,-2.6348],[182,-2.6348]]
       }
       index_03_l_ctrl.Rotation.Z {
        type=MovieSceneScriptingFloatChannel
        default=0
        keys [[42,-29.9933],[53,-3.0314],[72,0],[91,-29.9933],[165,-29.9933],[172,-2.7038],[177,-29.9933],[182,-29.9933]]
       }
       index_03_l_ctrl.Scale.X {type=MovieSceneScriptingFloatChannel, default=0, const [0,8]}
       index_03_l_ctrl.Scale.Y {type=MovieSceneScriptingFloatChannel, default=0, const [0,8]}
       index_03_l_ctrl.Scale.Z {type=MovieSceneScriptingFloatChannel, default=0, const [0,8]}
       middle_metacarpal_l_ctrl.Location.X {type=MovieSceneScriptingFloatChannel, default=0, thinned_from=8, keys [[42,0.0002],[53,0.0004],[72,0],[91,0.0002],[182,0.0002]]}
       middle_metacarpal_l_ctrl.Location.Y {type=MovieSceneScriptingFloatChannel, default=0, thinned_from=8, keys [[42,0.0001],[53,0.0002],[72,0],[165,0.0001],[172,0],[182,0.0001]]}
       middle_metacarpal_l_ctrl.Location.Z {type=MovieSceneScriptingFloatChannel, default=0, thinned_from=8, keys [[42,0],[182,0]]}
       middle_metacarpal_l_ctrl.Rotation.X {type=MovieSceneScriptingFloatChannel, default=0, thinned_from=8, keys [[42,0.0682],[53,0.0682],[72,0],[91,0.0682],[182,0.0682]]}
       middle_metacarpal_l_ctrl.Rotation.Y {
        type=MovieSceneScriptingFloatChannel
        default=0
        keys [[42,-0.5134],[53,-1.7387],[72,0],[91,-0.5134],[165,-0.5134],[172,-1.7387],[177,-0.5134],[182,-0.5134]]
       }
       middle_metacarpal_l_ctrl.Rotation.Z {type=MovieSceneScriptingFloatChannel, default=0, thinned_from=8, keys [[42,2.245],[53,2.245],[72,0],[91,2.245],[182,2.245]]}
       middle_metacarpal_l_ctrl.Scale.X {type=MovieSceneScriptingFloatChannel, default=0, const [0,8]}
       middle_metacarpal_l_ctrl.Scale.Y {type=MovieSceneScriptingFloatChannel, default=0, const [0,8]}
       middle_metacarpal_l_ctrl.Scale.Z {type=MovieSceneScriptingFloatChannel, default=0, const [0,8]}
       middle_01_l_ctrl.Location.X {type=MovieSceneScriptingFloatChannel, default=0, thinned_from=8, keys [[42,0.0003],[53,0.0011],[72,0],[91,0.0003],[182,0.0003]]}
       middle_01_l_ctrl.Location.Y {type=MovieSceneScriptingFloatChannel, default=0, thinned_from=8, keys [[42,0],[53,0.0006],[72,0],[182,0]]}
       middle_01_l_ctrl.Location.Z {type=MovieSceneScriptingFloatChannel, default=0, thinned_from=8, keys [[42,0],[182,0]]}
       middle_01_l_ctrl.Rotation.X {
        type=MovieSceneScriptingFloatChannel
        default=0
        keys [[42,12.3651],[53,-1.4224],[72,0],[91,12.3651],[165,12.3651],[172,8.4203],[177,12.3651],[182,12.3651]]
       }
       middle_01_l_ctrl.Rotation.Y {
        type=MovieSceneScriptingFloatChannel
        default=0
        keys [[42,6.4229],[53,15.7904],[72,0],[91,6.4229],[165,6.4229],[172,9.3939],[177,6.4229],[182,6.4229]]
       }
       middle_01_l_ctrl.Rotation.Z {
        type=MovieSceneScriptingFloatChannel
        default=0
        keys [[42,11.7771],[53,51.5344],[72,0],[91,11.7771],[165,11.7771],[172,69.0047],[177,11.7771],[182,11.7771]]
       }
       middle_01_l_ctrl.Scale.X {type=MovieSceneScriptingFloatChannel, default=0, const [0,8]}
       middle_01_l_ctrl.Scale.Y {type=MovieSceneScriptingFloatChannel, default=0, const [0,8]}
       middle_01_l_ctrl.Scale.Z {type=MovieSceneScriptingFloatChannel, default=0, const [0,8]}
       middle_02_l_ctrl.Location.X {
        type=MovieSceneScriptingFloatChannel
        default=0
        keys [[42,0.046],[53,0.0459],[72,0],[91,0.046],[165,0.046],[172,0.0452],[177,0.046],[182,0.046]]
       }
       middle_02_l_ctrl.Location.Y {
        type=MovieSceneScriptingFloatChannel
        default=0
        keys [[42,-0.0498],[53,-0.0497],[72,0],[91,-0.0498],[165,-0.0498],[172,-0.0506],[177,-0.0498],[182,-0.0498]]
       }
       middle_02_l_ctrl.Location.Z {
        type=MovieSceneScriptingFloatChannel
        default=0
        keys [[42,-0.067],[53,-0.067],[72,0],[91,-0.067],[165,-0.067],[172,-0.1146],[177,-0.067],[182,-0.067]]
       }
       middle_02_l_ctrl.Rotation.X {
        type=MovieSceneScriptingFloatChannel
        default=0
        keys [[42,3.1445],[53,3.2684],[72,0],[91,3.1445],[165,3.1445],[172,5.3143],[177,3.1445],[182,3.1445]]
       }
       middle_02_l_ctrl.Rotation.Y {
        type=MovieSceneScriptingFloatChannel
        default=0
        keys [[42,-8.6147],[53,-4.0798],[72,0],[91,-8.6147],[165,-8.6147],[172,-6.1393],[177,-8.6147],[182,-8.6147]]
       }
       middle_02_l_ctrl.Rotation.Z {
        type=MovieSceneScriptingFloatChannel
        default=0
        keys [[42,-38.3224],[53,29.6885],[72,0],[91,-38.3224],[165,-38.3224],[172,18.5698],[177,-38.3224],[182,-38.3224]]
       }
       middle_02_l_ctrl.Scale.X {type=MovieSceneScriptingFloatChannel, default=0, const [0,8]}
       middle_02_l_ctrl.Scale.Y {type=MovieSceneScriptingFloatChannel, default=0, const [0,8]}
       middle_02_l_ctrl.Scale.Z {type=MovieSceneScriptingFloatChannel, default=0, const [0,8]}
       middle_03_l_ctrl.Location.X {
        type=MovieSceneScriptingFloatChannel
        default=0
        keys [[42,-0.0201],[53,-0.02],[72,0],[91,-0.0201],[165,-0.0201],[172,-0.0206],[177,-0.0201],[182,-0.0201]]
       }
       middle_03_l_ctrl.Location.Y {type=MovieSceneScriptingFloatChannel, default=0, thinned_from=8, keys [[42,0.0583],[53,0.0582],[72,0],[91,0.0583],[182,0.0583]]}
       middle_03_l_ctrl.Location.Z {
        type=MovieSceneScriptingFloatChannel
        default=0
        keys [[42,-0.1263],[53,-0.1263],[72,0],[91,-0.1263],[165,-0.1263],[172,-0.1324],[177,-0.1263],[182,-0.1263]]
       }
       middle_03_l_ctrl.Rotation.X {
        type=MovieSceneScriptingFloatChannel
        default=0
        keys [[42,1.6984],[53,4.1009],[72,0],[91,1.6984],[165,1.6984],[172,2.8677],[177,1.6984],[182,1.6984]]
       }
       middle_03_l_ctrl.Rotation.Y {
        type=MovieSceneScriptingFloatChannel
        default=0
        keys [[42,-3.5476],[53,-3.0028],[72,0],[91,-3.5476],[165,-3.5476],[172,-5.0238],[177,-3.5476],[182,-3.5476]]
       }
       middle_03_l_ctrl.Rotation.Z {
        type=MovieSceneScriptingFloatChannel
        default=0
        keys [[42,-27.2791],[53,32.1846],[72,0],[91,-27.2791],[165,-27.2791],[172,24.1417],[177,-27.2791],[182,-27.2791]]
       }
       middle_03_l_ctrl.Scale.X {type=MovieSceneScriptingFloatChannel, default=0, const [0,8]}
       middle_03_l_ctrl.Scale.Y {type=MovieSceneScriptingFloatChannel, default=0, const [0,8]}
       middle_03_l_ctrl.Scale.Z {type=MovieSceneScriptingFloatChannel, default=0, const [0,8]}
       ring_metacarpal_l_ctrl.Location.X {type=MovieSceneScriptingFloatChannel, default=0, thinned_from=8, keys [[42,0.0001],[53,0.0005],[72,0],[91,0.0001],[182,0.0001]]}
       ring_metacarpal_l_ctrl.Location.Y {
        type=MovieSceneScriptingFloatChannel
        default=0
        thinned_from=8
        keys [[42,0.0001],[53,0.0002],[72,0],[91,0.0001],[165,0.0001],[172,0],[182,0.0001]]
       }
       ring_metacarpal_l_ctrl.Location.Z {type=MovieSceneScriptingFloatChannel, default=0, thinned_from=8, keys [[42,-0.0001],[53,0.0001],[91,-0.0001],[182,-0.0001]]}
       ring_metacarpal_l_ctrl.Rotation.X {type=MovieSceneScriptingFloatChannel, default=0, thinned_from=8, keys [[42,0.044],[53,0.044],[72,0],[91,0.044],[182,0.044]]}
       ring_metacarpal_l_ctrl.Rotation.Y {
        type=MovieSceneScriptingFloatChannel
        default=0
        keys [[42,-7.6476],[53,-3.1497],[72,0],[91,-7.6476],[165,-7.6476],[172,-3.1497],[177,-7.6476],[182,-7.6476]]
       }
       ring_metacarpal_l_ctrl.Rotation.Z {type=MovieSceneScriptingFloatChannel, default=0, thinned_from=8, keys [[42,0.8002],[53,0.8002],[72,0],[91,0.8002],[182,0.8002]]}
       ring_metacarpal_l_ctrl.Scale.X {type=MovieSceneScriptingFloatChannel, default=0, const [0,8]}
       ring_metacarpal_l_ctrl.Scale.Y {type=MovieSceneScriptingFloatChannel, default=0, const [0,8]}
       ring_metacarpal_l_ctrl.Scale.Z {type=MovieSceneScriptingFloatChannel, default=0, const [0,8]}
       ring_01_l_ctrl.Location.X {
        type=MovieSceneScriptingFloatChannel
        default=0
        thinned_from=8
        keys [[42,-0.0001],[53,0.0006],[72,0],[165,-0.0001],[172,0.0002],[177,-0.0001],[182,-0.0001]]
       }
       ring_01_l_ctrl.Location.Y {type=MovieSceneScriptingFloatChannel, default=0, thinned_from=8, keys [[42,0],[53,0.0007],[72,0],[182,0]]}
       ring_01_l_ctrl.Location.Z {type=MovieSceneScriptingFloatChannel, default=0, thinned_from=8, keys [[42,0],[182,0]]}
       ring_01_l_ctrl.Rotation.X {
        type=MovieSceneScriptingFloatChannel
        default=0
        keys [[42,8.2649],[53,7.6547],[72,0],[91,8.2649],[165,8.2649],[172,12.1103],[177,8.2649],[182,8.2649]]
       }
       ring_01_l_ctrl.Rotation.Y {
        type=MovieSceneScriptingFloatChannel
        default=0
        keys [[42,8.5727],[53,7.0064],[72,0],[91,8.5727],[165,8.5727],[172,3.7426],[177,8.5727],[182,8.5727]]
       }
       ring_01_l_ctrl.Rotation.Z {
        type=MovieSceneScriptingFloatChannel
        default=0
        keys [[42,0.1911],[53,34.8358],[72,0],[91,0.1911],[165,0.1911],[172,57.3472],[177,0.1911],[182,0.1911]]
       }
       ring_01_l_ctrl.Scale.X {type=MovieSceneScriptingFloatChannel, default=0, const [0,8]}
       ring_01_l_ctrl.Scale.Y {type=MovieSceneScriptingFloatChannel, default=0, const [0,8]}
       ring_01_l_ctrl.Scale.Z {type=MovieSceneScriptingFloatChannel, default=0, const [0,8]}
       ring_02_l_ctrl.Location.X {
        type=MovieSceneScriptingFloatChannel
        default=0
        thinned_from=8
        keys [[42,0.0003],[72,0],[91,0.0003],[165,0.0003],[172,0],[177,0.0003],[182,0.0003]]
       }
       ring_02_l_ctrl.Location.Y {type=MovieSceneScriptingFloatChannel, default=0, thinned_from=8, keys [[42,0.0001],[53,0.0002],[72,0],[182,0.0001]]}
       ring_02_l_ctrl.Location.Z {type=MovieSceneScriptingFloatChannel, default=0, thinned_from=8, keys [[42,0],[182,0]]}
       ring_02_l_ctrl.Rotation.X {
        type=MovieSceneScriptingFloatChannel
        default=0
        keys [[42,9.7758],[53,9.8337],[72,0],[91,9.7758],[165,9.7758],[172,8.3551],[177,9.7758],[182,9.7758]]
       }
       ring_02_l_ctrl.Rotation.Y {
        type=MovieSceneScriptingFloatChannel
        default=0
        keys [[42,-6.0123],[53,-5.4465],[72,0],[91,-6.0123],[165,-6.0123],[172,-5.9907],[177,-6.0123],[182,-6.0123]]
       }
       ring_02_l_ctrl.Rotation.Z {
        type=MovieSceneScriptingFloatChannel
        default=0
        keys [[42,-24.0533],[53,37.0406],[72,0],[91,-24.0533],[165,-24.0533],[172,38.2843],[177,-24.0533],[182,-24.0533]]
       }
       ring_02_l_ctrl.Scale.X {type=MovieSceneScriptingFloatChannel, default=0, const [0,8]}
       ring_02_l_ctrl.Scale.Y {type=MovieSceneScriptingFloatChannel, default=0, const [0,8]}
       ring_02_l_ctrl.Scale.Z {type=MovieSceneScriptingFloatChannel, default=0, const [0,8]}
       ring_03_l_ctrl.Location.X {type=MovieSceneScriptingFloatChannel, default=0, thinned_from=8, keys [[42,0],[53,0.0001],[182,0]]}
       ring_03_l_ctrl.Location.Y {type=MovieSceneScriptingFloatChannel, default=0, thinned_from=8, keys [[42,-0.0001],[53,-0.0002],[72,0],[165,-0.0001],[172,0],[182,-0.0001]]}
       ring_03_l_ctrl.Location.Z {type=MovieSceneScriptingFloatChannel, default=0, thinned_from=8, keys [[42,0],[182,0]]}
       ring_03_l_ctrl.Rotation.X {
        type=MovieSceneScriptingFloatChannel
        default=0
        keys [[42,3.7365],[53,3.8374],[72,0],[91,3.7365],[165,3.7365],[172,3.1846],[177,3.7365],[182,3.7365]]
       }
       ring_03_l_ctrl.Rotation.Y {
        type=MovieSceneScriptingFloatChannel
        default=0
        keys [[42,-4.3183],[53,-4.2812],[72,0],[91,-4.3183],[165,-4.3183],[172,-3.9076],[177,-4.3183],[182,-4.3183]]
       }
       ring_03_l_ctrl.Rotation.Z {
        type=MovieSceneScriptingFloatChannel
        default=0
        keys [[42,-32.4432],[53,46.4381],[72,0],[91,-32.4432],[165,-32.4432],[172,18.5607],[177,-32.4432],[182,-32.4432]]
       }
       ring_03_l_ctrl.Scale.X {type=MovieSceneScriptingFloatChannel, default=0, const [0,8]}
       ring_03_l_ctrl.Scale.Y {type=MovieSceneScriptingFloatChannel, default=0, const [0,8]}
       ring_03_l_ctrl.Scale.Z {type=MovieSceneScriptingFloatChannel, default=0, const [0,8]}
       pinky_metacarpal_l_ctrl.Location.X {type=MovieSceneScriptingFloatChannel, default=0, thinned_from=8, keys [[42,0.0001],[53,0.0003],[72,0],[182,0.0001]]}
       pinky_metacarpal_l_ctrl.Location.Y {type=MovieSceneScriptingFloatChannel, default=0, thinned_from=8, keys [[42,0],[182,0]]}
       pinky_metacarpal_l_ctrl.Location.Z {type=MovieSceneScriptingFloatChannel, default=0, thinned_from=8, keys [[42,0.0001],[53,0.0005],[72,0],[182,0.0001]]}
       pinky_metacarpal_l_ctrl.Rotation.X {
        type=MovieSceneScriptingFloatChannel
        default=0
        keys [[42,-7.3164],[53,0.5768],[72,0],[91,-7.3164],[165,-7.3164],[172,0.5768],[177,-7.3164],[182,-7.3164]]
       }
       pinky_metacarpal_l_ctrl.Rotation.Y {
        type=MovieSceneScriptingFloatChannel
        default=0
        keys [[42,-10.4177],[53,-7.1368],[72,0],[91,-10.4177],[165,-10.4177],[172,-7.1368],[177,-10.4177],[182,-10.4177]]
       }
       pinky_metacarpal_l_ctrl.Rotation.Z {
        type=MovieSceneScriptingFloatChannel
        default=0
        keys [[42,2.9375],[53,4.6328],[72,0],[91,2.9375],[165,2.9375],[172,4.6328],[177,2.9375],[182,2.9375]]
       }
       pinky_metacarpal_l_ctrl.Scale.X {type=MovieSceneScriptingFloatChannel, default=0, const [0,8]}
       pinky_metacarpal_l_ctrl.Scale.Y {type=MovieSceneScriptingFloatChannel, default=0, const [0,8]}
       pinky_metacarpal_l_ctrl.Scale.Z {type=MovieSceneScriptingFloatChannel, default=0, const [0,8]}
       pinky_01_l_ctrl.Location.X {type=MovieSceneScriptingFloatChannel, default=0, thinned_from=8, keys [[42,0.0002],[53,0.0008],[72,0],[91,0.0002],[182,0.0002]]}
       pinky_01_l_ctrl.Location.Y {
        type=MovieSceneScriptingFloatChannel
        default=0
        keys [[42,0.0003],[53,0.001],[72,0],[91,0.0003],[165,0.0003],[172,0],[177,0.0003],[182,0.0003]]
       }
       pinky_01_l_ctrl.Location.Z {type=MovieSceneScriptingFloatChannel, default=0, thinned_from=8, keys [[42,0],[182,0]]}
       pinky_01_l_ctrl.Rotation.X {
        type=MovieSceneScriptingFloatChannel
        default=0
        keys [[42,1.863],[53,1.5868],[72,0],[91,1.863],[165,1.863],[172,4.9113],[177,1.863],[182,1.863]]
       }
       pinky_01_l_ctrl.Rotation.Y {
        type=MovieSceneScriptingFloatChannel
        default=0
        keys [[42,13.39],[53,7.7118],[72,0],[91,13.39],[165,13.39],[172,7.9666],[177,13.39],[182,13.39]]
       }
       pinky_01_l_ctrl.Rotation.Z {
        type=MovieSceneScriptingFloatChannel
        default=0
        keys [[42,-39.447],[53,11.6001],[72,0],[91,-39.447],[165,-39.447],[172,31.4032],[177,-39.447],[182,-39.447]]
       }
       pinky_01_l_ctrl.Scale.X {type=MovieSceneScriptingFloatChannel, default=0, const [0,8]}
       pinky_01_l_ctrl.Scale.Y {type=MovieSceneScriptingFloatChannel, default=0, const [0,8]}
       pinky_01_l_ctrl.Scale.Z {type=MovieSceneScriptingFloatChannel, default=0, const [0,8]}
       pinky_02_l_ctrl.Location.X {type=MovieSceneScriptingFloatChannel, default=0, thinned_from=8, keys [[42,0],[182,0]]}
       pinky_02_l_ctrl.Location.Y {type=MovieSceneScriptingFloatChannel, default=0, thinned_from=8, keys [[42,0.0001],[53,0.0003],[72,0],[182,0.0001]]}
       pinky_02_l_ctrl.Location.Z {type=MovieSceneScriptingFloatChannel, default=0, thinned_from=8, keys [[42,0],[182,0]]}
       pinky_02_l_ctrl.Rotation.X {
        type=MovieSceneScriptingFloatChannel
        default=0
        keys [[42,11.2394],[53,11.3031],[72,0],[91,11.2394],[165,11.2394],[172,10.7826],[177,11.2394],[182,11.2394]]
       }
       pinky_02_l_ctrl.Rotation.Y {
        type=MovieSceneScriptingFloatChannel
        default=0
        keys [[42,6.8335],[53,6.845],[72,0],[91,6.8335],[165,6.8335],[172,5.451],[177,6.8335],[182,6.8335]]
       }
       pinky_02_l_ctrl.Rotation.Z {
        type=MovieSceneScriptingFloatChannel
        default=0
        keys [[42,-31.678],[53,22.0483],[72,0],[91,-31.678],[165,-31.678],[172,24.8653],[177,-31.678],[182,-31.678]]
       }
       pinky_02_l_ctrl.Scale.X {type=MovieSceneScriptingFloatChannel, default=0, const [0,8]}
       pinky_02_l_ctrl.Scale.Y {type=MovieSceneScriptingFloatChannel, default=0, const [0,8]}
       pinky_02_l_ctrl.Scale.Z {type=MovieSceneScriptingFloatChannel, default=0, const [0,8]}
       pinky_03_l_ctrl.Location.X {
        type=MovieSceneScriptingFloatChannel
        default=0
        thinned_from=8
        keys [[42,0.0002],[72,0],[91,0.0002],[165,0.0002],[172,0],[177,0.0002],[182,0.0002]]
       }
       pinky_03_l_ctrl.Location.Y {
        type=MovieSceneScriptingFloatChannel
        default=0
        keys [[42,-0.0002],[53,-0.0003],[72,0],[91,-0.0002],[165,-0.0002],[172,0],[177,-0.0002],[182,-0.0002]]
       }
       pinky_03_l_ctrl.Location.Z {type=MovieSceneScriptingFloatChannel, default=0, thinned_from=8, keys [[42,0],[182,0]]}
       pinky_03_l_ctrl.Rotation.X {
        type=MovieSceneScriptingFloatChannel
        default=0
        keys [[42,6.5917],[53,7.6153],[72,0],[91,6.5917],[165,6.5917],[172,6.9833],[177,6.5917],[182,6.5917]]
       }
       pinky_03_l_ctrl.Rotation.Y {
        type=MovieSceneScriptingFloatChannel
        default=0
        keys [[42,3.6359],[53,4.4837],[72,0],[91,3.6359],[165,3.6359],[172,2.7587],[177,3.6359],[182,3.6359]]
       }
       pinky_03_l_ctrl.Rotation.Z {
        type=MovieSceneScriptingFloatChannel
        default=0
        keys [[42,-32.3795],[53,40.9066],[72,0],[91,-32.3795],[165,-32.3795],[172,23.422],[177,-32.3795],[182,-32.3795]]
       }
       pinky_03_l_ctrl.Scale.X {type=MovieSceneScriptingFloatChannel, default=0, const [0,8]}
       pinky_03_l_ctrl.Scale.Y {type=MovieSceneScriptingFloatChannel, default=0, const [0,8]}
       pinky_03_l_ctrl.Scale.Z {type=MovieSceneScriptingFloatChannel, default=0, const [0,8]}
       hand_l_ik_ctrl.Location.X {type=MovieSceneScriptingFloatChannel, default=0}
       hand_l_ik_ctrl.Location.Y {type=MovieSceneScriptingFloatChannel, default=0}
       hand_l_ik_ctrl.Location.Z {type=MovieSceneScriptingFloatChannel, default=0}
       hand_l_ik_ctrl.Rotation.X {type=MovieSceneScriptingFloatChannel, default=0}
       hand_l_ik_ctrl.Rotation.Y {type=MovieSceneScriptingFloatChannel, default=0}
       hand_l_ik_ctrl.Rotation.Z {type=MovieSceneScriptingFloatChannel, default=0}
       hand_l_ik_ctrl.Scale.X {type=MovieSceneScriptingFloatChannel, default=0}
       hand_l_ik_ctrl.Scale.Y {type=MovieSceneScriptingFloatChannel, default=0}
       hand_l_ik_ctrl.Scale.Z {type=MovieSceneScriptingFloatChannel, default=0}
       Softness {type=MovieSceneScriptingFloatChannel, default=0}
       hand_l_ikorient_ctrl.Rotation.Roll {type=MovieSceneScriptingFloatChannel, default=0}
       hand_l_ikorient_ctrl.Rotation.Pitch {type=MovieSceneScriptingFloatChannel, default=0}
       hand_l_ikorient_ctrl.Rotation.Yaw {type=MovieSceneScriptingFloatChannel, default=0}
       arm_l_pv_ik_ctrl.Location.X {type=MovieSceneScriptingFloatChannel, default=0}
       arm_l_pv_ik_ctrl.Location.Y {type=MovieSceneScriptingFloatChannel, default=0}
       arm_l_pv_ik_ctrl.Location.Z {type=MovieSceneScriptingFloatChannel, default=0}
       arm_l_pv_ik_ctrl.Rotation.X {type=MovieSceneScriptingFloatChannel, default=0}
       arm_l_pv_ik_ctrl.Rotation.Y {type=MovieSceneScriptingFloatChannel, default=0}
       arm_l_pv_ik_ctrl.Rotation.Z {type=MovieSceneScriptingFloatChannel, default=0}
       arm_l_pv_ik_ctrl.Scale.X {type=MovieSceneScriptingFloatChannel, default=0}
       arm_l_pv_ik_ctrl.Scale.Y {type=MovieSceneScriptingFloatChannel, default=0}
       arm_l_pv_ik_ctrl.Scale.Z {type=MovieSceneScriptingFloatChannel, default=0}
       upperarm_r_fk_ctrl.Location.X {type=MovieSceneScriptingFloatChannel, default=0, const [0,7]}
       upperarm_r_fk_ctrl.Location.Y {type=MovieSceneScriptingFloatChannel, default=0, const [0,7]}
       upperarm_r_fk_ctrl.Location.Z {type=MovieSceneScriptingFloatChannel, default=0, const [0,7]}
       upperarm_r_fk_ctrl.Rotation.X {
        type=MovieSceneScriptingFloatChannel
        default=0
        keys [[94,0],[109,-12.9872],[163,-12.8624],[174,26.2683],[175,26.2818],[178,26.2089],[183,0]]
       }
       upperarm_r_fk_ctrl.Rotation.Y {
        type=MovieSceneScriptingFloatChannel
        default=0
        keys [[94,0],[109,-26.1905],[163,-25.9388],[174,-1.3794],[175,-1.3672],[178,-1.3632],[183,0]]
       }
       upperarm_r_fk_ctrl.Rotation.Z {
        type=MovieSceneScriptingFloatChannel
        default=0
        thinned_from=7
        keys [[94,0],[109,-0.0785],[163,-0.0778],[174,-0.0143],[178,-0.0142],[183,0]]
       }
       upperarm_r_fk_ctrl.Scale.X {type=MovieSceneScriptingFloatChannel, default=0, const [0,7]}
       upperarm_r_fk_ctrl.Scale.Y {type=MovieSceneScriptingFloatChannel, default=0, const [0,7]}
       upperarm_r_fk_ctrl.Scale.Z {type=MovieSceneScriptingFloatChannel, default=0, const [0,7]}
       lowerarm_r_fk_ctrl.Location.X {type=MovieSceneScriptingFloatChannel, default=0, const [0,4]}
       lowerarm_r_fk_ctrl.Location.Y {type=MovieSceneScriptingFloatChannel, default=0, const [0,4]}
       lowerarm_r_fk_ctrl.Location.Z {type=MovieSceneScriptingFloatChannel, default=0, const [0,4]}
       lowerarm_r_fk_ctrl.Rotation.X {type=MovieSceneScriptingFloatChannel, default=0, const [0,4]}
       lowerarm_r_fk_ctrl.Rotation.Y {type=MovieSceneScriptingFloatChannel, default=0, const [0,4]}
       lowerarm_r_fk_ctrl.Rotation.Z {type=MovieSceneScriptingFloatChannel, default=0, const [0,4]}
       lowerarm_r_fk_ctrl.Scale.X {type=MovieSceneScriptingFloatChannel, default=0, const [0,4]}
       lowerarm_r_fk_ctrl.Scale.Y {type=MovieSceneScriptingFloatChannel, default=0, const [0,4]}
       lowerarm_r_fk_ctrl.Scale.Z {type=MovieSceneScriptingFloatChannel, default=0, const [0,4]}
       hand_r_fk_ctrl.Location.X {type=MovieSceneScriptingFloatChannel, default=0, const [0,11]}
       hand_r_fk_ctrl.Location.Y {type=MovieSceneScriptingFloatChannel, default=0, const [0,11]}
       hand_r_fk_ctrl.Location.Z {type=MovieSceneScriptingFloatChannel, default=0, const [0,11]}
       hand_r_fk_ctrl.Rotation.X {
        type=MovieSceneScriptingFloatChannel
        default=0
        thinned_from=11
        keys [[133,0],[137,0.8905],[143,15.7695],[148,15.863],[165,15.863],[169,0],[174,-5.0992],[175,-5.1027],[178,-5.0886],[183,0]]
       }
       hand_r_fk_ctrl.Rotation.Y {
        type=MovieSceneScriptingFloatChannel
        default=0
        thinned_from=11
        keys [[133,0],[137,15.5197],[143,-3.5575],[148,-0.278],[165,-0.278],[169,0],[174,5.6669],[175,5.674],[178,5.6582],[183,0]]
       }
       hand_r_fk_ctrl.Rotation.Z {
        type=MovieSceneScriptingFloatChannel
        default=0
        thinned_from=11
        keys [[133,0],[137,5.0263],[143,19.4709],[148,21.022],[165,21.022],[169,0],[174,23.2724],[175,23.3027],[178,23.238],[183,0]]
       }
       hand_r_fk_ctrl.Scale.X {type=MovieSceneScriptingFloatChannel, default=0, const [0,11]}
       hand_r_fk_ctrl.Scale.Y {type=MovieSceneScriptingFloatChannel, default=0, const [0,11]}
       hand_r_fk_ctrl.Scale.Z {type=MovieSceneScriptingFloatChannel, default=0, const [0,11]}
       thumb_01_r_ctrl.Location.X {type=MovieSceneScriptingFloatChannel, default=0, thinned_from=8, keys [[53,0.0004],[110,0],[164,0],[170,0.0005],[177,0.0004]]}
       thumb_01_r_ctrl.Location.Y {type=MovieSceneScriptingFloatChannel, default=0, thinned_from=8, keys [[53,0.0004],[110,0],[164,0],[177,0.0004]]}
       thumb_01_r_ctrl.Location.Z {type=MovieSceneScriptingFloatChannel, default=0, thinned_from=8, keys [[53,0.0002],[164,0],[170,0.0006],[177,0.0002]]}
       thumb_01_r_ctrl.Rotation.X {
        type=MovieSceneScriptingFloatChannel
        default=0
        keys [[53,-15.3564],[62,-15.3564],[98,-2.3994],[104,20.6234],[110,0],[164,0],[170,12.6648],[177,-15.3564]]
       }
       thumb_01_r_ctrl.Rotation.Y {
        type=MovieSceneScriptingFloatChannel
        default=0
        keys [[53,0.0903],[62,0.0903],[98,0.0141],[104,2.1097],[110,0],[164,0],[170,-3.425],[177,0.0903]]
       }
       thumb_01_r_ctrl.Rotation.Z {
        type=MovieSceneScriptingFloatChannel
        default=0
        keys [[53,-1.428],[62,-1.428],[98,-0.2231],[104,-8.3607],[110,0],[164,0],[170,-5.8844],[177,-1.428]]
       }
       thumb_01_r_ctrl.Scale.X {type=MovieSceneScriptingFloatChannel, default=0, const [0,8]}
       thumb_01_r_ctrl.Scale.Y {type=MovieSceneScriptingFloatChannel, default=0, const [0,8]}
       thumb_01_r_ctrl.Scale.Z {type=MovieSceneScriptingFloatChannel, default=0, const [0,8]}
       thumb_02_r_ctrl.Location.X {type=MovieSceneScriptingFloatChannel, default=0, thinned_from=9, keys [[53,0],[98,0],[104,0.0001],[164,0],[170,0.0013],[177,0]]}
       thumb_02_r_ctrl.Location.Y {type=MovieSceneScriptingFloatChannel, default=0, thinned_from=9, keys [[53,0.0006],[110,0],[164,0],[170,0.0005],[177,0.0006]]}
       thumb_02_r_ctrl.Location.Z {type=MovieSceneScriptingFloatChannel, default=0, thinned_from=9, keys [[53,0.0002],[164,0],[170,0.0002],[177,0.0002]]}
       thumb_02_r_ctrl.Rotation.X {
        type=MovieSceneScriptingFloatChannel
        default=0
        thinned_from=9
        keys [[53,1.3317],[62,1.3317],[98,0.2081],[104,1.3796],[110,0],[164,0],[170,0.9984],[177,1.3317]]
       }
       thumb_02_r_ctrl.Rotation.Y {
        type=MovieSceneScriptingFloatChannel
        default=0
        thinned_from=9
        keys [[53,-6.4378],[62,-6.4378],[98,-1.0059],[104,4.3129],[110,0],[164,0],[170,-0.3832],[177,-6.4378]]
       }
       thumb_02_r_ctrl.Rotation.Z {
        type=MovieSceneScriptingFloatChannel
        default=0
        thinned_from=9
        keys [[53,-29.5839],[62,-29.5839],[98,-4.6225],[104,4.1364],[110,0],[164,0],[170,-4.5332],[177,-29.5839]]
       }
       thumb_02_r_ctrl.Scale.X {type=MovieSceneScriptingFloatChannel, default=0, const [0,9]}
       thumb_02_r_ctrl.Scale.Y {type=MovieSceneScriptingFloatChannel, default=0, const [0,9]}
       thumb_02_r_ctrl.Scale.Z {type=MovieSceneScriptingFloatChannel, default=0, const [0,9]}
       thumb_03_r_ctrl.Location.X {type=MovieSceneScriptingFloatChannel, default=0, thinned_from=9, keys [[53,0.0001],[98,0],[104,0.0001],[164,0],[170,0.0004],[177,0.0001]]}
       thumb_03_r_ctrl.Location.Y {type=MovieSceneScriptingFloatChannel, default=0, thinned_from=9, keys [[53,0.0001],[170,0],[177,0.0001]]}
       thumb_03_r_ctrl.Location.Z {type=MovieSceneScriptingFloatChannel, default=0, thinned_from=9, keys [[53,0.0001],[177,0.0001]]}
       thumb_03_r_ctrl.Rotation.X {
        type=MovieSceneScriptingFloatChannel
        default=0
        thinned_from=9
        keys [[53,1.7965],[62,1.7965],[98,0.2807],[104,-1.215],[110,0],[164,0],[170,0.7579],[177,1.7965]]
       }
       thumb_03_r_ctrl.Rotation.Y {
        type=MovieSceneScriptingFloatChannel
        default=0
        thinned_from=9
        keys [[53,-0.1955],[62,-0.1955],[98,-0.0306],[104,-5.8066],[110,0],[164,0],[170,-3.33],[177,-0.1955]]
       }
       thumb_03_r_ctrl.Rotation.Z {
        type=MovieSceneScriptingFloatChannel
        default=0
        thinned_from=9
        keys [[53,-53.2697],[62,-53.2697],[98,-8.3234],[104,16.2338],[110,0],[164,0],[170,-28.7414],[177,-53.2697]]
       }
       thumb_03_r_ctrl.Scale.X {type=MovieSceneScriptingFloatChannel, default=0, const [0,9]}
       thumb_03_r_ctrl.Scale.Y {type=MovieSceneScriptingFloatChannel, default=0, const [0,9]}
       thumb_03_r_ctrl.Scale.Z {type=MovieSceneScriptingFloatChannel, default=0, const [0,9]}
       index_metacarpal_r_ctrl.Location.X {type=MovieSceneScriptingFloatChannel, default=0, thinned_from=9, keys [[53,0.0003],[110,0],[168,0],[170,0.0007],[177,0.0003]]}
       index_metacarpal_r_ctrl.Location.Y {type=MovieSceneScriptingFloatChannel, default=0, thinned_from=9, keys [[53,0],[168,0],[170,0.0003],[177,0]]}
       index_metacarpal_r_ctrl.Location.Z {
        type=MovieSceneScriptingFloatChannel
        default=0
        thinned_from=9
        keys [[53,-0.0007],[62,-0.0007],[98,-0.0001],[168,0],[170,-0.0007],[177,-0.0007]]
       }
       index_metacarpal_r_ctrl.Rotation.X {
        type=MovieSceneScriptingFloatChannel
        default=0
        thinned_from=9
        keys [[53,-0.01],[62,-0.01],[98,-0.0016],[104,7.2204],[110,0],[168,-0.0001],[170,-0.01],[177,-0.01]]
       }
       index_metacarpal_r_ctrl.Rotation.Y {
        type=MovieSceneScriptingFloatChannel
        default=0
        keys [[53,4.5784],[62,4.5784],[98,0.7154],[104,3.1387],[110,0],[164,0],[168,0.0189],[170,3.2472],[177,4.5784]]
       }
       index_metacarpal_r_ctrl.Rotation.Z {
        type=MovieSceneScriptingFloatChannel
        default=0
        keys [[53,0.74],[62,0.74],[98,0.1156],[104,0.74],[110,0],[164,0],[168,0.0048],[170,0.74],[177,0.74]]
       }
       index_metacarpal_r_ctrl.Scale.X {type=MovieSceneScriptingFloatChannel, default=0, const [0,9]}
       index_metacarpal_r_ctrl.Scale.Y {type=MovieSceneScriptingFloatChannel, default=0, const [0,9]}
       index_metacarpal_r_ctrl.Scale.Z {type=MovieSceneScriptingFloatChannel, default=0, const [0,9]}
       index_01_r_ctrl.Location.X {
        type=MovieSceneScriptingFloatChannel
        default=0
        thinned_from=12
        keys [[53,0.0007],[62,0.0007],[98,0.0001],[104,0.0002],[110,0],[168,0],[170,0.0024],[177,0.0007]]
       }
       index_01_r_ctrl.Location.Y {
        type=MovieSceneScriptingFloatChannel
        default=0
        thinned_from=12
        keys [[53,0.0009],[62,0.0009],[98,0.0001],[104,-0.0002],[110,0],[168,0],[170,-0.0001],[177,0.0009]]
       }
       index_01_r_ctrl.Location.Z {type=MovieSceneScriptingFloatChannel, default=0, thinned_from=12, keys [[53,-0.0005],[104,0.0001],[168,0],[170,-0.0007],[177,-0.0005]]}
       index_01_r_ctrl.Rotation.X {
        type=MovieSceneScriptingFloatChannel
        default=0
        thinned_from=12
        keys [[53,-6.5565],[62,-6.5565],[98,-1.1736],[104,-11.4277],[110,0],[120,0.1473],[164,0.1473],[168,0.0427],[170,-15.9444],[177,-6.5565]]
       }
       index_01_r_ctrl.Rotation.Y {
        type=MovieSceneScriptingFloatChannel
        default=0
        thinned_from=12
        keys [[53,-0.7753],[62,-0.7753],[98,-0.1211],[104,-1.0397],[110,0],[120,-0.0423],[164,-0.0423],[168,0.007],[170,7.5551],[177,-0.7753]]
       }
       index_01_r_ctrl.Rotation.Z {
        type=MovieSceneScriptingFloatChannel
        default=0
        keys [[53,14.9073],[62,14.9073],[98,4.9104],[104,21.45],[110,0],[120,-7.2713],[134,-7.2713],[143,22.251],[164,22.251],[168,13.8777],[170,50.1766],[177,14.9073]]
       }
       index_01_r_ctrl.Scale.X {type=MovieSceneScriptingFloatChannel, default=0, const [0,12]}
       index_01_r_ctrl.Scale.Y {type=MovieSceneScriptingFloatChannel, default=0, const [0,12]}
       index_01_r_ctrl.Scale.Z {type=MovieSceneScriptingFloatChannel, default=0, const [0,12]}
       index_02_r_ctrl.Location.X {
        type=MovieSceneScriptingFloatChannel
        default=0
        thinned_from=10
        keys [[53,0.0001],[98,0],[104,0.0002],[120,0],[168,0],[170,0.0012],[177,0.0001]]
       }
       index_02_r_ctrl.Location.Y {
        type=MovieSceneScriptingFloatChannel
        default=0
        thinned_from=10
        keys [[53,-0.0001],[98,0],[104,0.0002],[110,0],[168,0],[170,0.001],[177,-0.0001]]
       }
       index_02_r_ctrl.Location.Z {type=MovieSceneScriptingFloatChannel, default=0, thinned_from=10, keys [[53,-0.0001],[168,0],[170,0.0006],[177,-0.0001]]}
       index_02_r_ctrl.Rotation.X {
        type=MovieSceneScriptingFloatChannel
        default=0
        keys [[53,2.3069],[62,2.3069],[98,0.3604],[104,0.584],[110,0],[120,0.0211],[164,0.0211],[168,0.0314],[170,1.7486],[177,2.3069]]
       }
       index_02_r_ctrl.Rotation.Y {
        type=MovieSceneScriptingFloatChannel
        default=0
        keys [[53,0.5596],[62,0.5596],[98,0.1447],[104,0.2944],[110,0],[120,-0.0566],[164,-0.0566],[168,-0.0523],[170,0.5937],[177,0.5596]]
       }
       index_02_r_ctrl.Rotation.Z {
        type=MovieSceneScriptingFloatChannel
        default=0
        keys [[53,-37.6994],[62,-37.6994],[98,-5.8905],[104,-11.1084],[110,0],[120,-7.2697],[164,-7.2697],[168,-7.0732],[170,22.9788],[177,-37.6994]]
       }
       index_02_r_ctrl.Scale.X {type=MovieSceneScriptingFloatChannel, default=0, const [0,10]}
       index_02_r_ctrl.Scale.Y {type=MovieSceneScriptingFloatChannel, default=0, const [0,10]}
       index_02_r_ctrl.Scale.Z {type=MovieSceneScriptingFloatChannel, default=0, const [0,10]}
       index_03_r_ctrl.Location.X {
        type=MovieSceneScriptingFloatChannel
        default=0
        thinned_from=9
        keys [[53,0.0005],[98,0.0001],[104,0.0004],[110,0],[168,0],[170,0.0025],[177,0.0005]]
       }
       index_03_r_ctrl.Location.Y {
        type=MovieSceneScriptingFloatChannel
        default=0
        thinned_from=9
        keys [[53,0.0001],[98,0],[104,0.0003],[110,0],[168,0],[170,0.0013],[177,0.0001]]
       }
       index_03_r_ctrl.Location.Z {type=MovieSceneScriptingFloatChannel, default=0, thinned_from=9, keys [[53,0],[98,0],[104,-0.0001],[110,0],[168,0],[170,0.0009],[177,0]]}
       index_03_r_ctrl.Rotation.X {
        type=MovieSceneScriptingFloatChannel
        default=0
        keys [[53,0.049],[62,0.049],[98,0.0077],[104,0.3679],[110,0],[164,0],[168,-0.0002],[170,-0.023],[177,0.049]]
       }
       index_03_r_ctrl.Rotation.Y {
        type=MovieSceneScriptingFloatChannel
        default=0
        keys [[53,0.8364],[62,0.8364],[98,0.1307],[104,1.3958],[110,0],[164,0],[168,0.0079],[170,1.2091],[177,0.8364]]
       }
       index_03_r_ctrl.Rotation.Z {
        type=MovieSceneScriptingFloatChannel
        default=0
        keys [[53,-31.4722],[62,-31.4722],[98,-4.9175],[104,17.5255],[110,0],[164,0],[168,-0.0866],[170,-16.2282],[177,-31.4722]]
       }
       index_03_r_ctrl.Scale.X {type=MovieSceneScriptingFloatChannel, default=0, const [0,9]}
       index_03_r_ctrl.Scale.Y {type=MovieSceneScriptingFloatChannel, default=0, const [0,9]}
       index_03_r_ctrl.Scale.Z {type=MovieSceneScriptingFloatChannel, default=0, const [0,9]}
       middle_metacarpal_r_ctrl.Location.X {
        type=MovieSceneScriptingFloatChannel
        default=0
        thinned_from=9
        keys [[53,0.0003],[98,0],[104,0.0001],[110,0],[168,0],[170,0.0007],[177,0.0003]]
       }
       middle_metacarpal_r_ctrl.Location.Y {
        type=MovieSceneScriptingFloatChannel
        default=0
        thinned_from=9
        keys [[53,0.0001],[98,0],[104,-0.0002],[110,0],[168,0],[170,-0.0002],[177,0.0001]]
       }
       middle_metacarpal_r_ctrl.Location.Z {type=MovieSceneScriptingFloatChannel, default=0, thinned_from=9, keys [[53,-0.0003],[110,0],[168,0],[170,-0.0003],[177,-0.0003]]}
       middle_metacarpal_r_ctrl.Rotation.X {
        type=MovieSceneScriptingFloatChannel
        default=0
        keys [[53,-0.07],[62,-0.07],[98,-0.0109],[104,-0.07],[110,0],[164,0],[168,-0.0005],[170,-0.07],[177,-0.07]]
       }
       middle_metacarpal_r_ctrl.Rotation.Y {
        type=MovieSceneScriptingFloatChannel
        default=0
        keys [[53,-0.5102],[62,-0.5102],[98,-0.0797],[104,-1.7355],[110,0],[164,0],[168,-0.0113],[170,-1.7355],[177,-0.5102]]
       }
       middle_metacarpal_r_ctrl.Rotation.Z {
        type=MovieSceneScriptingFloatChannel
        default=0
        keys [[53,2.244],[62,2.244],[98,0.3506],[104,2.244],[110,0],[164,0],[168,0.0146],[170,2.244],[177,2.244]]
       }
       middle_metacarpal_r_ctrl.Scale.X {type=MovieSceneScriptingFloatChannel, default=0, const [0,9]}
       middle_metacarpal_r_ctrl.Scale.Y {type=MovieSceneScriptingFloatChannel, default=0, const [0,9]}
       middle_metacarpal_r_ctrl.Scale.Z {type=MovieSceneScriptingFloatChannel, default=0, const [0,9]}
       middle_01_r_ctrl.Location.X {
        type=MovieSceneScriptingFloatChannel
        default=0
        thinned_from=11
        keys [[53,0.0006],[98,0.0001],[104,0.0002],[110,0],[168,0],[170,0.0021],[177,0.0006]]
       }
       middle_01_r_ctrl.Location.Y {
        type=MovieSceneScriptingFloatChannel
        default=0
        thinned_from=11
        keys [[53,0.0008],[62,0.0008],[104,-0.0001],[168,0],[170,-0.0003],[177,0.0008]]
       }
       middle_01_r_ctrl.Location.Z {
        type=MovieSceneScriptingFloatChannel
        default=0
        thinned_from=11
        keys [[53,-0.0003],[98,-0.0001],[104,0.0001],[168,0],[170,-0.0004],[177,-0.0003]]
       }
       middle_01_r_ctrl.Rotation.X {
        type=MovieSceneScriptingFloatChannel
        default=0
        keys [[53,13.145],[62,13.145],[98,2.0539],[104,-6.6657],[110,0],[134,0],[143,-2.3054],[164,-2.3054],[168,-0.0041],[170,-0.6386],[177,13.145]]
       }
       middle_01_r_ctrl.Rotation.Y {
        type=MovieSceneScriptingFloatChannel
        default=0
        keys [[53,2.8962],[62,2.8962],[98,0.4525],[104,1.5211],[110,0],[134,0],[143,-2.3212],[164,-2.3212],[168,0.0642],[170,9.8735],[177,2.8962]]
       }
       middle_01_r_ctrl.Rotation.Z {
        type=MovieSceneScriptingFloatChannel
        default=0
        keys [[53,12.3434],[62,12.3434],[98,1.9287],[104,8.1485],[110,0],[134,0],[143,28.9197],[164,28.9197],[168,15.7198],[170,48.8302],[177,12.3434]]
       }
       middle_01_r_ctrl.Scale.X {type=MovieSceneScriptingFloatChannel, default=0, const [0,11]}
       middle_01_r_ctrl.Scale.Y {type=MovieSceneScriptingFloatChannel, default=0, const [0,11]}
       middle_01_r_ctrl.Scale.Z {type=MovieSceneScriptingFloatChannel, default=0, const [0,11]}
       middle_02_r_ctrl.Location.X {
        type=MovieSceneScriptingFloatChannel
        default=0
        thinned_from=9
        keys [[53,0.0002],[98,0],[104,0.0002],[110,0],[168,0],[170,0.0008],[177,0.0002]]
       }
       middle_02_r_ctrl.Location.Y {
        type=MovieSceneScriptingFloatChannel
        default=0
        thinned_from=9
        keys [[53,-0.0001],[98,0],[104,0.0002],[110,0],[168,0],[170,0.0014],[177,-0.0001]]
       }
       middle_02_r_ctrl.Location.Z {type=MovieSceneScriptingFloatChannel, default=0, thinned_from=9, keys [[53,0],[168,0],[170,0.0008],[177,0]]}
       middle_02_r_ctrl.Rotation.X {
        type=MovieSceneScriptingFloatChannel
        default=0
        keys [[53,1.8181],[62,1.8181],[98,0.2841],[104,-3.4508],[110,0],[164,0],[168,-0.0152],[170,-2.3468],[177,1.8181]]
       }
       middle_02_r_ctrl.Rotation.Y {
        type=MovieSceneScriptingFloatChannel
        default=0
        keys [[53,-12.2069],[62,-12.2069],[98,-1.9073],[104,-3.3016],[110,0],[164,0],[168,-0.0235],[170,-4.7428],[177,-12.2069]]
       }
       middle_02_r_ctrl.Rotation.Z {
        type=MovieSceneScriptingFloatChannel
        default=0
        keys [[53,-30.8012],[62,-30.8012],[98,-4.8127],[104,3.1978],[110,0],[164,0],[168,0.0649],[170,9.9893],[177,-30.8012]]
       }
       middle_02_r_ctrl.Scale.X {type=MovieSceneScriptingFloatChannel, default=0, const [0,9]}
       middle_02_r_ctrl.Scale.Y {type=MovieSceneScriptingFloatChannel, default=0, const [0,9]}
       middle_02_r_ctrl.Scale.Z {type=MovieSceneScriptingFloatChannel, default=0, const [0,9]}
       middle_03_r_ctrl.Location.X {
        type=MovieSceneScriptingFloatChannel
        default=0
        thinned_from=9
        keys [[53,0.0005],[98,0.0001],[104,0.0002],[110,0],[168,0],[170,0.0025],[177,0.0005]]
       }
       middle_03_r_ctrl.Location.Y {
        type=MovieSceneScriptingFloatChannel
        default=0
        thinned_from=9
        keys [[53,0.0001],[98,0],[104,0.0001],[110,0],[168,0],[170,0.0011],[177,0.0001]]
       }
       middle_03_r_ctrl.Location.Z {type=MovieSceneScriptingFloatChannel, default=0, thinned_from=9, keys [[53,0],[98,0],[104,-0.0001],[110,0],[168,0],[170,0.001],[177,0]]}
       middle_03_r_ctrl.Rotation.X {
        type=MovieSceneScriptingFloatChannel
        default=0
        keys [[53,2.0623],[62,2.0623],[98,0.3222],[104,3.9145],[110,0],[164,0],[168,0.019],[170,2.9243],[177,2.0623]]
       }
       middle_03_r_ctrl.Rotation.Y {
        type=MovieSceneScriptingFloatChannel
        default=0
        keys [[53,-4.3096],[62,-4.3096],[98,-0.6734],[104,-2.5544],[110,0],[164,-0.0017],[168,-0.0114],[170,-1.9046],[177,-4.3096]]
       }
       middle_03_r_ctrl.Rotation.Z {
        type=MovieSceneScriptingFloatChannel
        default=0
        keys [[53,-21.4228],[62,-21.4228],[98,-3.3493],[104,50.0517],[110,0],[164,0.0106],[168,0.1555],[170,22.3149],[177,-21.4228]]
       }
       middle_03_r_ctrl.Scale.X {type=MovieSceneScriptingFloatChannel, default=0, const [0,9]}
       middle_03_r_ctrl.Scale.Y {type=MovieSceneScriptingFloatChannel, default=0, const [0,9]}
       middle_03_r_ctrl.Scale.Z {type=MovieSceneScriptingFloatChannel, default=0, const [0,9]}
       ring_metacarpal_r_ctrl.Location.X {type=MovieSceneScriptingFloatChannel, default=0, thinned_from=8, keys [[53,0.0002],[110,0],[164,0],[170,0.0006],[177,0.0002]]}
       ring_metacarpal_r_ctrl.Location.Y {
        type=MovieSceneScriptingFloatChannel
        default=0
        thinned_from=8
        keys [[53,0.0001],[98,0],[104,-0.0003],[110,0],[164,0],[170,-0.0005],[177,0.0001]]
       }
       ring_metacarpal_r_ctrl.Location.Z {type=MovieSceneScriptingFloatChannel, default=0, thinned_from=8, keys [[53,0.0002],[164,0],[177,0.0002]]}
       ring_metacarpal_r_ctrl.Rotation.X {type=MovieSceneScriptingFloatChannel, default=0, keys [[53,0.04],[62,0.04],[98,0.0063],[104,-4.6478],[110,0],[164,0],[170,0.04],[177,0.04]]}
       ring_metacarpal_r_ctrl.Rotation.Y {
        type=MovieSceneScriptingFloatChannel
        default=0
        keys [[53,-7.6511],[62,-7.6511],[98,-1.1955],[104,-3.1532],[110,0],[164,0],[170,-3.1532],[177,-7.6511]]
       }
       ring_metacarpal_r_ctrl.Rotation.Z {type=MovieSceneScriptingFloatChannel, default=0, keys [[53,0.8],[62,0.8],[98,0.125],[104,0.8],[110,0],[164,0],[170,0.8],[177,0.8]]}
       ring_metacarpal_r_ctrl.Scale.X {type=MovieSceneScriptingFloatChannel, default=0, const [0,8]}
       ring_metacarpal_r_ctrl.Scale.Y {type=MovieSceneScriptingFloatChannel, default=0, const [0,8]}
       ring_metacarpal_r_ctrl.Scale.Z {type=MovieSceneScriptingFloatChannel, default=0, const [0,8]}
       ring_01_r_ctrl.Location.X {
        type=MovieSceneScriptingFloatChannel
        default=0
        thinned_from=8
        keys [[53,0.0004],[98,0.0001],[104,0.0003],[110,0],[164,0],[170,0.0021],[177,0.0004]]
       }
       ring_01_r_ctrl.Location.Y {
        type=MovieSceneScriptingFloatChannel
        default=0
        thinned_from=8
        keys [[53,0.0008],[62,0.0008],[104,-0.0001],[164,0],[170,-0.0001],[177,0.0008]]
       }
       ring_01_r_ctrl.Location.Z {
        type=MovieSceneScriptingFloatChannel
        default=0
        thinned_from=8
        keys [[53,-0.0003],[98,0],[104,0.0001],[110,0],[164,0],[170,-0.0003],[177,-0.0003]]
       }
       ring_01_r_ctrl.Rotation.X {
        type=MovieSceneScriptingFloatChannel
        default=0
        keys [[53,5.3059],[62,5.3059],[98,0.829],[104,-7.0521],[110,0],[164,0],[170,5.6652],[177,5.3059]]
       }
       ring_01_r_ctrl.Rotation.Y {
        type=MovieSceneScriptingFloatChannel
        default=0
        keys [[53,6.1451],[62,6.1451],[98,0.9602],[104,3.951],[110,0],[164,0],[170,4.6548],[177,6.1451]]
       }
       ring_01_r_ctrl.Rotation.Z {
        type=MovieSceneScriptingFloatChannel
        default=0
        keys [[53,-1.0427],[62,-1.0427],[98,-0.1629],[104,-13.963],[110,0],[164,0],[170,34.9561],[177,-1.0427]]
       }
       ring_01_r_ctrl.Scale.X {type=MovieSceneScriptingFloatChannel, default=0, const [0,8]}
       ring_01_r_ctrl.Scale.Y {type=MovieSceneScriptingFloatChannel, default=0, const [0,8]}
       ring_01_r_ctrl.Scale.Z {type=MovieSceneScriptingFloatChannel, default=0, const [0,8]}
       ring_02_r_ctrl.Location.X {
        type=MovieSceneScriptingFloatChannel
        default=0
        thinned_from=9
        keys [[53,0.0001],[98,0],[104,0.0002],[110,0],[164,0],[170,0.0003],[177,0.0001]]
       }
       ring_02_r_ctrl.Location.Y {type=MovieSceneScriptingFloatChannel, default=0, thinned_from=9, keys [[53,-0.0001],[164,0],[170,0.001],[177,-0.0001]]}
       ring_02_r_ctrl.Location.Z {type=MovieSceneScriptingFloatChannel, default=0, thinned_from=9, keys [[53,-0.0001],[164,0],[170,0.0003],[177,-0.0001]]}
       ring_02_r_ctrl.Rotation.X {
        type=MovieSceneScriptingFloatChannel
        default=0
        thinned_from=9
        keys [[53,11.3992],[62,11.3992],[98,1.7811],[104,9.2417],[110,0],[164,0],[170,8.0465],[177,11.3992]]
       }
       ring_02_r_ctrl.Rotation.Y {
        type=MovieSceneScriptingFloatChannel
        default=0
        thinned_from=9
        keys [[53,-6.0548],[62,-6.0548],[98,-0.9461],[104,-8.5169],[110,0],[164,0],[170,-6.9687],[177,-6.0548]]
       }
       ring_02_r_ctrl.Rotation.Z {
        type=MovieSceneScriptingFloatChannel
        default=0
        keys [[53,-2.6945],[62,-2.6945],[98,-0.421],[104,39.0837],[110,0],[122,11.3842],[164,11.3842],[170,32.9191],[177,-2.6945]]
       }
       ring_02_r_ctrl.Scale.X {type=MovieSceneScriptingFloatChannel, default=0, const [0,9]}
       ring_02_r_ctrl.Scale.Y {type=MovieSceneScriptingFloatChannel, default=0, const [0,9]}
       ring_02_r_ctrl.Scale.Z {type=MovieSceneScriptingFloatChannel, default=0, const [0,9]}
       ring_03_r_ctrl.Location.X {
        type=MovieSceneScriptingFloatChannel
        default=0
        thinned_from=9
        keys [[53,0.0004],[98,0.0001],[104,0.0004],[110,0],[164,0],[170,0.0011],[177,0.0004]]
       }
       ring_03_r_ctrl.Location.Y {type=MovieSceneScriptingFloatChannel, default=0, thinned_from=9, keys [[53,0.0002],[98,0],[104,0.0002],[110,0],[177,0.0002]]}
       ring_03_r_ctrl.Location.Z {type=MovieSceneScriptingFloatChannel, default=0, thinned_from=9, keys [[53,0],[98,0],[104,-0.0002],[110,0],[164,0],[170,0.0006],[177,0]]}
       ring_03_r_ctrl.Rotation.X {
        type=MovieSceneScriptingFloatChannel
        default=0
        thinned_from=9
        keys [[53,2.2485],[62,2.2485],[98,0.3513],[104,1.6156],[110,0],[164,0],[170,1.1565],[177,2.2485]]
       }
       ring_03_r_ctrl.Rotation.Y {
        type=MovieSceneScriptingFloatChannel
        default=0
        thinned_from=9
        keys [[53,-4.8435],[62,-4.8435],[98,-0.7568],[104,-5.7497],[110,0],[164,0],[170,-3.4656],[177,-4.8435]]
       }
       ring_03_r_ctrl.Rotation.Z {
        type=MovieSceneScriptingFloatChannel
        default=0
        keys [[53,-31.029],[62,-31.029],[98,-4.8483],[104,32.1542],[110,0],[122,11.3842],[164,11.3842],[170,34.8023],[177,-31.029]]
       }
       ring_03_r_ctrl.Scale.X {type=MovieSceneScriptingFloatChannel, default=0, const [0,9]}
       ring_03_r_ctrl.Scale.Y {type=MovieSceneScriptingFloatChannel, default=0, const [0,9]}
       ring_03_r_ctrl.Scale.Z {type=MovieSceneScriptingFloatChannel, default=0, const [0,9]}
       pinky_metacarpal_r_ctrl.Location.X {
        type=MovieSceneScriptingFloatChannel
        default=0
        thinned_from=8
        keys [[53,0.0002],[98,0],[104,0.0002],[110,0],[164,0],[170,0.0005],[177,0.0002]]
       }
       pinky_metacarpal_r_ctrl.Location.Y {type=MovieSceneScriptingFloatChannel, default=0, thinned_from=8, keys [[53,0.0002],[104,-0.0001],[164,0],[170,-0.0008],[177,0.0002]]}
       pinky_metacarpal_r_ctrl.Location.Z {
        type=MovieSceneScriptingFloatChannel
        default=0
        thinned_from=8
        keys [[53,0.0006],[98,0.0001],[104,0.0002],[110,0],[164,0],[170,0.0007],[177,0.0006]]
       }
       pinky_metacarpal_r_ctrl.Rotation.X {type=MovieSceneScriptingFloatChannel, default=0, keys [[53,0.58],[62,0.58],[98,0.0906],[104,-4.8181],[110,0],[164,0],[170,0.58],[177,0.58]]}
       pinky_metacarpal_r_ctrl.Rotation.Y {
        type=MovieSceneScriptingFloatChannel
        default=0
        keys [[53,-7.1566],[62,-7.1566],[98,-1.1182],[104,-7.1566],[110,0],[164,0],[170,-7.1566],[177,-7.1566]]
       }
       pinky_metacarpal_r_ctrl.Rotation.Z {type=MovieSceneScriptingFloatChannel, default=0, keys [[53,4.63],[62,4.63],[98,0.7234],[104,4.63],[110,0],[164,0],[170,4.63],[177,4.63]]}
       pinky_metacarpal_r_ctrl.Scale.X {type=MovieSceneScriptingFloatChannel, default=0, const [0,8]}
       pinky_metacarpal_r_ctrl.Scale.Y {type=MovieSceneScriptingFloatChannel, default=0, const [0,8]}
       pinky_metacarpal_r_ctrl.Scale.Z {type=MovieSceneScriptingFloatChannel, default=0, const [0,8]}
       pinky_01_r_ctrl.Location.X {
        type=MovieSceneScriptingFloatChannel
        default=0
        thinned_from=8
        keys [[53,0.0003],[98,0],[104,0.0003],[110,0],[164,0],[170,0.002],[177,0.0003]]
       }
       pinky_01_r_ctrl.Location.Y {type=MovieSceneScriptingFloatChannel, default=0, thinned_from=8, keys [[53,0.0009],[62,0.0009],[104,0],[170,0],[177,0.0009]]}
       pinky_01_r_ctrl.Location.Z {
        type=MovieSceneScriptingFloatChannel
        default=0
        thinned_from=8
        keys [[53,-0.0003],[98,0],[104,0.0002],[110,0],[164,0],[170,-0.0003],[177,-0.0003]]
       }
       pinky_01_r_ctrl.Rotation.X {
        type=MovieSceneScriptingFloatChannel
        default=0
        keys [[53,7.1303],[62,7.1303],[98,1.1141],[104,-0.9331],[110,0],[164,0],[170,7.4361],[177,7.1303]]
       }
       pinky_01_r_ctrl.Rotation.Y {
        type=MovieSceneScriptingFloatChannel
        default=0
        keys [[53,11.3646],[62,11.3646],[98,1.7757],[104,11.3612],[110,0],[164,0],[170,5.0257],[177,11.3646]]
       }
       pinky_01_r_ctrl.Rotation.Z {
        type=MovieSceneScriptingFloatChannel
        default=0
        keys [[53,-26.6863],[62,-26.6863],[98,-4.1697],[104,-8.555],[110,0],[164,0],[170,27.5842],[177,-26.6863]]
       }
       pinky_01_r_ctrl.Scale.X {type=MovieSceneScriptingFloatChannel, default=0, const [0,8]}
       pinky_01_r_ctrl.Scale.Y {type=MovieSceneScriptingFloatChannel, default=0, const [0,8]}
       pinky_01_r_ctrl.Scale.Z {type=MovieSceneScriptingFloatChannel, default=0, const [0,8]}
       pinky_02_r_ctrl.Location.X {type=MovieSceneScriptingFloatChannel, default=0, thinned_from=9, keys [[53,0.0001],[104,0.0001],[164,0],[170,0.0002],[177,0.0001]]}
       pinky_02_r_ctrl.Location.Y {type=MovieSceneScriptingFloatChannel, default=0, thinned_from=9, keys [[53,-0.0001],[164,0],[170,0.0007],[177,-0.0001]]}
       pinky_02_r_ctrl.Location.Z {type=MovieSceneScriptingFloatChannel, default=0, thinned_from=9, keys [[53,-0.0001],[170,0.0001],[177,-0.0001]]}
       pinky_02_r_ctrl.Rotation.X {
        type=MovieSceneScriptingFloatChannel
        default=0
        thinned_from=9
        keys [[53,15.5939],[62,15.5939],[98,2.4366],[104,14.1847],[110,0],[164,0],[170,13.5456],[177,15.5939]]
       }
       pinky_02_r_ctrl.Rotation.Y {
        type=MovieSceneScriptingFloatChannel
        default=0
        thinned_from=9
        keys [[53,1.39],[62,1.39],[98,0.2172],[104,-3.4537],[110,0],[164,0],[170,-0.9719],[177,1.39]]
       }
       pinky_02_r_ctrl.Rotation.Z {
        type=MovieSceneScriptingFloatChannel
        default=0
        keys [[53,-8.5975],[62,-8.5975],[98,-1.3434],[104,40.0658],[110,0],[122,11.3842],[164,11.3842],[170,27.6287],[177,-8.5975]]
       }
       pinky_02_r_ctrl.Scale.X {type=MovieSceneScriptingFloatChannel, default=0, const [0,9]}
       pinky_02_r_ctrl.Scale.Y {type=MovieSceneScriptingFloatChannel, default=0, const [0,9]}
       pinky_02_r_ctrl.Scale.Z {type=MovieSceneScriptingFloatChannel, default=0, const [0,9]}
       pinky_03_r_ctrl.Location.X {
        type=MovieSceneScriptingFloatChannel
        default=0
        thinned_from=9
        keys [[53,0.0004],[98,0.0001],[104,0.0002],[110,0],[164,0],[170,0.0005],[177,0.0004]]
       }
       pinky_03_r_ctrl.Location.Y {type=MovieSceneScriptingFloatChannel, default=0, thinned_from=9, keys [[53,0.0001],[170,-0.0001],[177,0.0001]]}
       pinky_03_r_ctrl.Location.Z {type=MovieSceneScriptingFloatChannel, default=0, thinned_from=9, keys [[53,0],[164,0],[170,0.0005],[177,0]]}
       pinky_03_r_ctrl.Rotation.X {
        type=MovieSceneScriptingFloatChannel
        default=0
        thinned_from=9
        keys [[53,7.4814],[62,7.4814],[98,1.169],[104,8.0814],[110,0],[164,0],[170,6.097],[177,7.4814]]
       }
       pinky_03_r_ctrl.Rotation.Y {
        type=MovieSceneScriptingFloatChannel
        default=0
        thinned_from=9
        keys [[53,-2.0303],[62,-2.0303],[98,-0.3172],[104,-3.6053],[110,0],[164,0],[170,-1.1714],[177,-2.0303]]
       }
       pinky_03_r_ctrl.Rotation.Z {
        type=MovieSceneScriptingFloatChannel
        default=0
        keys [[53,-25.024],[62,-25.024],[98,-3.91],[104,26.348],[110,0],[122,11.3842],[164,11.3842],[170,34.0362],[177,-25.024]]
       }
       pinky_03_r_ctrl.Scale.X {type=MovieSceneScriptingFloatChannel, default=0, const [0,9]}
       pinky_03_r_ctrl.Scale.Y {type=MovieSceneScriptingFloatChannel, default=0, const [0,9]}
       pinky_03_r_ctrl.Scale.Z {type=MovieSceneScriptingFloatChannel, default=0, const [0,9]}
       hand_r_ik_ctrl.Location.X {type=MovieSceneScriptingFloatChannel, default=0}
       hand_r_ik_ctrl.Location.Y {type=MovieSceneScriptingFloatChannel, default=0}
       hand_r_ik_ctrl.Location.Z {type=MovieSceneScriptingFloatChannel, default=0}
       hand_r_ik_ctrl.Rotation.X {type=MovieSceneScriptingFloatChannel, default=0}
       hand_r_ik_ctrl.Rotation.Y {type=MovieSceneScriptingFloatChannel, default=0}
       hand_r_ik_ctrl.Rotation.Z {type=MovieSceneScriptingFloatChannel, default=0}
       hand_r_ik_ctrl.Scale.X {type=MovieSceneScriptingFloatChannel, default=0}
       hand_r_ik_ctrl.Scale.Y {type=MovieSceneScriptingFloatChannel, default=0}
       hand_r_ik_ctrl.Scale.Z {type=MovieSceneScriptingFloatChannel, default=0}
       Softness_2 {type=MovieSceneScriptingFloatChannel, default=0}
       hand_r_ikorient_ctrl.Rotation.Roll {type=MovieSceneScriptingFloatChannel, default=0}
       hand_r_ikorient_ctrl.Rotation.Pitch {type=MovieSceneScriptingFloatChannel, default=0}
       hand_r_ikorient_ctrl.Rotation.Yaw {type=MovieSceneScriptingFloatChannel, default=0}
       arm_r_pv_ik_ctrl.Location.X {type=MovieSceneScriptingFloatChannel, default=0}
       arm_r_pv_ik_ctrl.Location.Y {type=MovieSceneScriptingFloatChannel, default=0}
       arm_r_pv_ik_ctrl.Location.Z {type=MovieSceneScriptingFloatChannel, default=0}
       arm_r_pv_ik_ctrl.Rotation.X {type=MovieSceneScriptingFloatChannel, default=0}
       arm_r_pv_ik_ctrl.Rotation.Y {type=MovieSceneScriptingFloatChannel, default=0}
       arm_r_pv_ik_ctrl.Rotation.Z {type=MovieSceneScriptingFloatChannel, default=0}
       arm_r_pv_ik_ctrl.Scale.X {type=MovieSceneScriptingFloatChannel, default=0}
       arm_r_pv_ik_ctrl.Scale.Y {type=MovieSceneScriptingFloatChannel, default=0}
       arm_r_pv_ik_ctrl.Scale.Z {type=MovieSceneScriptingFloatChannel, default=0}
       body_ctrl.Location.X {type=MovieSceneScriptingFloatChannel, default=0, thinned_from=8, keys [[55,0],[183,0],[184,-11.2102]]}
       body_ctrl.Location.Y {
        type=MovieSceneScriptingFloatChannel
        default=-248
        thinned_from=9
        keys [[47,-965.1103],[55,-248],[68,-63.3387],[102,0],[183,0],[184,52.7481]]
       }
       body_ctrl.Location.Z {
        type=MovieSceneScriptingFloatChannel
        default=0
        keys [[55,0],[68,9.4601],[102,1.4956],[137,8.7536],[167,6.7758],[179,6.7544],[183,0.702],[184,24.4724]]
       }
       body_ctrl.Rotation.X {type=MovieSceneScriptingFloatChannel, default=0, const [0,8]}
       body_ctrl.Rotation.Y {type=MovieSceneScriptingFloatChannel, default=0, const [0,8]}
       body_ctrl.Rotation.Z {type=MovieSceneScriptingFloatChannel, default=0, const [0,8]}
       body_ctrl.Scale.X {type=MovieSceneScriptingFloatChannel, default=0, const [0,8]}
       body_ctrl.Scale.Y {type=MovieSceneScriptingFloatChannel, default=0, const [0,8]}
       body_ctrl.Scale.Z {type=MovieSceneScriptingFloatChannel, default=0, const [0,8]}
       hips_ctrl.Location.X {type=MovieSceneScriptingFloatChannel, default=0}
       hips_ctrl.Location.Y {type=MovieSceneScriptingFloatChannel, default=0}
       hips_ctrl.Location.Z {type=MovieSceneScriptingFloatChannel, default=0}
       hips_ctrl.Rotation.X {type=MovieSceneScriptingFloatChannel, default=0}
       hips_ctrl.Rotation.Y {type=MovieSceneScriptingFloatChannel, default=0}
       hips_ctrl.Rotation.Z {type=MovieSceneScriptingFloatChannel, default=0}
       hips_ctrl.Scale.X {type=MovieSceneScriptingFloatChannel, default=0}
       hips_ctrl.Scale.Y {type=MovieSceneScriptingFloatChannel, default=0}
       hips_ctrl.Scale.Z {type=MovieSceneScriptingFloatChannel, default=0}
       hips_tan_ctrl.Location.X {type=MovieSceneScriptingFloatChannel, default=0}
       hips_tan_ctrl.Location.Y {type=MovieSceneScriptingFloatChannel, default=0}
       hips_tan_ctrl.Location.Z {type=MovieSceneScriptingFloatChannel, default=0}
       hips_tan_ctrl.Rotation.X {type=MovieSceneScriptingFloatChannel, default=0}
       hips_tan_ctrl.Rotation.Y {type=MovieSceneScriptingFloatChannel, default=0}
       hips_tan_ctrl.Rotation.Z {type=MovieSceneScriptingFloatChannel, default=0}
       hips_tan_ctrl.Scale.X {type=MovieSceneScriptingFloatChannel, default=0}
       hips_tan_ctrl.Scale.Y {type=MovieSceneScriptingFloatChannel, default=0}
       hips_tan_ctrl.Scale.Z {type=MovieSceneScriptingFloatChannel, default=0}
       spine_01_ctrl.Location.X {type=MovieSceneScriptingFloatChannel, default=0, const [0,3]}
       spine_01_ctrl.Location.Y {type=MovieSceneScriptingFloatChannel, default=0, const [0,3]}
       spine_01_ctrl.Location.Z {type=MovieSceneScriptingFloatChannel, default=0, const [0,3]}
       spine_01_ctrl.Rotation.X {type=MovieSceneScriptingFloatChannel, default=0, keys [[167,0],[177,2.3072],[182,0]]}
       spine_01_ctrl.Rotation.Y {type=MovieSceneScriptingFloatChannel, default=0, const [0,3]}
       spine_01_ctrl.Rotation.Z {type=MovieSceneScriptingFloatChannel, default=0, const [0,3]}
       spine_01_ctrl.Scale.X {type=MovieSceneScriptingFloatChannel, default=0, const [0,3]}
       spine_01_ctrl.Scale.Y {type=MovieSceneScriptingFloatChannel, default=0, const [0,3]}
       spine_01_ctrl.Scale.Z {type=MovieSceneScriptingFloatChannel, default=0, const [0,3]}
       spine_02_ctrl.Location.X {type=MovieSceneScriptingFloatChannel, default=0, const [0,3]}
       spine_02_ctrl.Location.Y {type=MovieSceneScriptingFloatChannel, default=0, const [0,3]}
       spine_02_ctrl.Location.Z {type=MovieSceneScriptingFloatChannel, default=0, const [0,3]}
       spine_02_ctrl.Rotation.X {type=MovieSceneScriptingFloatChannel, default=0, keys [[167,0],[177,3.6516],[182,0]]}
       spine_02_ctrl.Rotation.Y {type=MovieSceneScriptingFloatChannel, default=0, const [0,3]}
       spine_02_ctrl.Rotation.Z {type=MovieSceneScriptingFloatChannel, default=0, const [0,3]}
       spine_02_ctrl.Scale.X {type=MovieSceneScriptingFloatChannel, default=0, const [0,3]}
       spine_02_ctrl.Scale.Y {type=MovieSceneScriptingFloatChannel, default=0, const [0,3]}
       spine_02_ctrl.Scale.Z {type=MovieSceneScriptingFloatChannel, default=0, const [0,3]}
       spine_03_ctrl.Location.X {type=MovieSceneScriptingFloatChannel, default=0, keys [[167,0],[177,-0.0077],[182,0]]}
       spine_03_ctrl.Location.Y {type=MovieSceneScriptingFloatChannel, default=0, keys [[167,0],[177,0.0477],[182,0]]}
       spine_03_ctrl.Location.Z {type=MovieSceneScriptingFloatChannel, default=0, keys [[167,0],[177,-1.0963],[182,0]]}
       spine_03_ctrl.Rotation.X {type=MovieSceneScriptingFloatChannel, default=0, keys [[167,0],[177,3.6516],[182,0]]}
       spine_03_ctrl.Rotation.Y {type=MovieSceneScriptingFloatChannel, default=0, const [0,3]}
       spine_03_ctrl.Rotation.Z {type=MovieSceneScriptingFloatChannel, default=0, const [0,3]}
       spine_03_ctrl.Scale.X {type=MovieSceneScriptingFloatChannel, default=0, const [0,3]}
       spine_03_ctrl.Scale.Y {type=MovieSceneScriptingFloatChannel, default=0, const [0,3]}
       spine_03_ctrl.Scale.Z {type=MovieSceneScriptingFloatChannel, default=0, const [0,3]}
       spine_04_ctrl.Location.X {type=MovieSceneScriptingFloatChannel, default=0, const [0,5]}
       spine_04_ctrl.Location.Y {type=MovieSceneScriptingFloatChannel, default=0, const [0,5]}
       spine_04_ctrl.Location.Z {type=MovieSceneScriptingFloatChannel, default=0, const [0,5]}
       spine_04_ctrl.Rotation.X {type=MovieSceneScriptingFloatChannel, default=0, const [0,5]}
       spine_04_ctrl.Rotation.Y {type=MovieSceneScriptingFloatChannel, default=0, const [0,5]}
       spine_04_ctrl.Rotation.Z {type=MovieSceneScriptingFloatChannel, default=0, const [0,5]}
       spine_04_ctrl.Scale.X {type=MovieSceneScriptingFloatChannel, default=0, const [0,5]}
       spine_04_ctrl.Scale.Y {type=MovieSceneScriptingFloatChannel, default=0, const [0,5]}
       spine_04_ctrl.Scale.Z {type=MovieSceneScriptingFloatChannel, default=0, const [0,5]}
       head_ctrl.Location.X {
        type=MovieSceneScriptingFloatChannel
        default=0
        thinned_from=9
        keys [[48,0],[162,0],[166,-0.0011],[173,-0.0096],[179,-0.0107],[182,-0.0132]]
       }
       head_ctrl.Location.Y {
        type=MovieSceneScriptingFloatChannel
        default=0
        thinned_from=9
        keys [[48,0],[162,0],[166,-0.0006],[173,-0.0051],[179,-0.0057],[182,-0.007]]
       }
       head_ctrl.Location.Z {type=MovieSceneScriptingFloatChannel, default=0, thinned_from=9, keys [[48,0],[162,0],[166,0.008],[173,0.0725],[179,0.0809],[182,0.0997]]}
       head_ctrl.Rotation.X {
        type=MovieSceneScriptingFloatChannel
        default=0
        keys [[48,0],[70,8.8694],[90,-3.7214],[109,-6.9165],[162,-6.9133],[166,-5.0198],[173,19.106],[179,23.0839],[182,-3.7666]]
       }
       head_ctrl.Rotation.Y {
        type=MovieSceneScriptingFloatChannel
        default=0
        keys [[48,0],[70,-3.13],[90,-3.0831],[109,-3.0713],[162,-3.0735],[166,-3.5048],[173,-4.7126],[179,-4.378],[182,-5.6933]]
       }
       head_ctrl.Rotation.Z {
        type=MovieSceneScriptingFloatChannel
        default=0
        keys [[48,0],[70,8.3538],[90,7.3721],[109,7.1232],[162,7.1327],[166,20.75],[173,52.2587],[179,49.9171],[182,3.715]]
       }
       head_ctrl.Scale.X {type=MovieSceneScriptingFloatChannel, default=0, const [0,9]}
       head_ctrl.Scale.Y {type=MovieSceneScriptingFloatChannel, default=0, const [0,9]}
       head_ctrl.Scale.Z {type=MovieSceneScriptingFloatChannel, default=0, const [0,9]}
       clavicle_l_ctrl.Location.X {type=MovieSceneScriptingFloatChannel, default=0}
       clavicle_l_ctrl.Location.Y {type=MovieSceneScriptingFloatChannel, default=0}
       clavicle_l_ctrl.Location.Z {type=MovieSceneScriptingFloatChannel, default=0}
       clavicle_l_ctrl.Rotation.X {type=MovieSceneScriptingFloatChannel, default=0}
       clavicle_l_ctrl.Rotation.Y {type=MovieSceneScriptingFloatChannel, default=0}
       clavicle_l_ctrl.Rotation.Z {type=MovieSceneScriptingFloatChannel, default=0}
       clavicle_l_ctrl.Scale.X {type=MovieSceneScriptingFloatChannel, default=0}
       clavicle_l_ctrl.Scale.Y {type=MovieSceneScriptingFloatChannel, default=0}
       clavicle_l_ctrl.Scale.Z {type=MovieSceneScriptingFloatChannel, default=0}
       clavicle_r_ctrl.Location.X {type=MovieSceneScriptingFloatChannel, default=0}
       clavicle_r_ctrl.Location.Y {type=MovieSceneScriptingFloatChannel, default=0}
       clavicle_r_ctrl.Location.Z {type=MovieSceneScriptingFloatChannel, default=0}
       clavicle_r_ctrl.Rotation.X {type=MovieSceneScriptingFloatChannel, default=0}
       clavicle_r_ctrl.Rotation.Y {type=MovieSceneScriptingFloatChannel, default=0}
       clavicle_r_ctrl.Rotation.Z {type=MovieSceneScriptingFloatChannel, default=0}
       clavicle_r_ctrl.Scale.X {type=MovieSceneScriptingFloatChannel, default=0}
       clavicle_r_ctrl.Scale.Y {type=MovieSceneScriptingFloatChannel, default=0}
       clavicle_r_ctrl.Scale.Z {type=MovieSceneScriptingFloatChannel, default=0}
       neck_01_ctrl.Location.X {type=MovieSceneScriptingFloatChannel, default=0, thinned_from=8, keys [[70,0],[182,0]]}
       neck_01_ctrl.Location.Y {type=MovieSceneScriptingFloatChannel, default=0, thinned_from=8, keys [[70,0],[182,0]]}
       neck_01_ctrl.Location.Z {type=MovieSceneScriptingFloatChannel, default=0, thinned_from=8, keys [[70,0],[182,0]]}
       neck_01_ctrl.Rotation.X {
        type=MovieSceneScriptingFloatChannel
        default=0
        keys [[70,9.0689],[100,0.0135],[103,0.0135],[132,0],[164,0],[169,6.793],[179,11.5166],[182,-7.7773]]
       }
       neck_01_ctrl.Rotation.Y {
        type=MovieSceneScriptingFloatChannel
        default=0
        keys [[70,-5.1107],[100,-5.2019],[103,-5.1906],[132,0],[164,0],[169,-3.3095],[179,-5.6108],[182,-0.7778]]
       }
       neck_01_ctrl.Rotation.Z {
        type=MovieSceneScriptingFloatChannel
        default=0
        keys [[70,0.102],[100,0.1039],[103,0.1036],[132,0],[164,0],[169,-1.8356],[179,-3.5378],[182,-6.8015]]
       }
       neck_01_ctrl.Scale.X {type=MovieSceneScriptingFloatChannel, default=0, const [0,8]}
       neck_01_ctrl.Scale.Y {type=MovieSceneScriptingFloatChannel, default=0, const [0,8]}
       neck_01_ctrl.Scale.Z {type=MovieSceneScriptingFloatChannel, default=0, const [0,8]}
       neck_02_ctrl.Location.X {type=MovieSceneScriptingFloatChannel, default=0, thinned_from=4, keys [[164,0],[182,0]]}
       neck_02_ctrl.Location.Y {type=MovieSceneScriptingFloatChannel, default=0, keys [[164,0],[169,-0.0787],[179,-0.1334],[182,0]]}
       neck_02_ctrl.Location.Z {type=MovieSceneScriptingFloatChannel, default=0, keys [[164,0],[169,-0.2351],[179,-0.3985],[182,0]]}
       neck_02_ctrl.Rotation.X {type=MovieSceneScriptingFloatChannel, default=0, keys [[164,0],[169,9.757],[179,16.5417],[182,2.0006]]}
       neck_02_ctrl.Rotation.Y {type=MovieSceneScriptingFloatChannel, default=0, keys [[164,0],[169,-2.1003],[179,-3.6258],[182,-3.7798]]}
       neck_02_ctrl.Rotation.Z {type=MovieSceneScriptingFloatChannel, default=0, keys [[164,0],[169,3.8321],[179,6.4969],[182,-6.294]]}
       neck_02_ctrl.Scale.X {type=MovieSceneScriptingFloatChannel, default=0, const [0,4]}
       neck_02_ctrl.Scale.Y {type=MovieSceneScriptingFloatChannel, default=0, const [0,4]}
       neck_02_ctrl.Scale.Z {type=MovieSceneScriptingFloatChannel, default=0, const [0,4]}
       head_ik_ctrl.Location.X {type=MovieSceneScriptingFloatChannel, default=0}
       head_ik_ctrl.Location.Y {type=MovieSceneScriptingFloatChannel, default=0}
       head_ik_ctrl.Location.Z {type=MovieSceneScriptingFloatChannel, default=0}
       head_ik_ctrl.Rotation.X {type=MovieSceneScriptingFloatChannel, default=0}
       head_ik_ctrl.Rotation.Y {type=MovieSceneScriptingFloatChannel, default=0}
       head_ik_ctrl.Rotation.Z {type=MovieSceneScriptingFloatChannel, default=0}
       head_ik_ctrl.Scale.X {type=MovieSceneScriptingFloatChannel, default=0}
       head_ik_ctrl.Scale.Y {type=MovieSceneScriptingFloatChannel, default=0}
       head_ik_ctrl.Scale.Z {type=MovieSceneScriptingFloatChannel, default=0}
       chest_ctrl.Location.X {type=MovieSceneScriptingFloatChannel, default=0}
       chest_ctrl.Location.Y {type=MovieSceneScriptingFloatChannel, default=0}
       chest_ctrl.Location.Z {type=MovieSceneScriptingFloatChannel, default=0}
       chest_ctrl.Rotation.X {type=MovieSceneScriptingFloatChannel, default=0}
       chest_ctrl.Rotation.Y {type=MovieSceneScriptingFloatChannel, default=0}
       chest_ctrl.Rotation.Z {type=MovieSceneScriptingFloatChannel, default=0}
       chest_ctrl.Scale.X {type=MovieSceneScriptingFloatChannel, default=0}
       chest_ctrl.Scale.Y {type=MovieSceneScriptingFloatChannel, default=0}
       chest_ctrl.Scale.Z {type=MovieSceneScriptingFloatChannel, default=0}
       chest_tan_ctrl.Location.X {type=MovieSceneScriptingFloatChannel, default=0}
       chest_tan_ctrl.Location.Y {type=MovieSceneScriptingFloatChannel, default=0}
       chest_tan_ctrl.Location.Z {type=MovieSceneScriptingFloatChannel, default=0}
       chest_tan_ctrl.Rotation.X {type=MovieSceneScriptingFloatChannel, default=0}
       chest_tan_ctrl.Rotation.Y {type=MovieSceneScriptingFloatChannel, default=0}
       chest_tan_ctrl.Rotation.Z {type=MovieSceneScriptingFloatChannel, default=0}
       chest_tan_ctrl.Scale.X {type=MovieSceneScriptingFloatChannel, default=0}
       chest_tan_ctrl.Scale.Y {type=MovieSceneScriptingFloatChannel, default=0}
       chest_tan_ctrl.Scale.Z {type=MovieSceneScriptingFloatChannel, default=0}
      ]
     }
    ]
   }
   {type=MovieSceneSpawnTrack, sections [{type=MovieSceneSpawnSection, channels [None {type=MovieSceneScriptingBoolChannel, default=true}]}]}
   {
    type=MovieScene3DTransformTrack
    sections [{
      type=MovieScene3DTransformSection
      channels [
       Location.X {type=MovieSceneScriptingDoubleChannel, default=0, const [0,1]}
       Location.Y {type=MovieSceneScriptingDoubleChannel, default=0, const [0,1]}
       Location.Z {type=MovieSceneScriptingDoubleChannel, default=0, const [0,1]}
       Rotation.X {type=MovieSceneScriptingDoubleChannel, default=0, const [0,1]}
       Rotation.Y {type=MovieSceneScriptingDoubleChannel, default=0, const [0,1]}
       Rotation.Z {type=MovieSceneScriptingDoubleChannel, default=0, const [0,1]}
       Scale.X {type=MovieSceneScriptingDoubleChannel, default=1, const [1,1]}
       Scale.Y {type=MovieSceneScriptingDoubleChannel, default=1, const [1,1]}
       Scale.Z {type=MovieSceneScriptingDoubleChannel, default=1, const [1,1]}
      ]
     }]
   }
  ]
 }
]
display_rate {denominator=1, numerator=30}
tick_resolution {denominator=1, numerator=24000}
tracks [{type=MovieSceneCameraCutTrack, sections [{type=MovieSceneCameraCutSection, end_frame=184, start_frame=0}]}]
