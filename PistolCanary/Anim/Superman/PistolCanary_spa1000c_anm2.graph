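# Level Sequence that animates the Superman cinematic NPC through an airborne performance with sweeping travel, banking and turns, over roughly frames 1800 to 2820.
# The possessed Superman_BP carries a skeletal animation section plus a Fortnite biped Control Rig parameter track.
# Dense per-frame keys drive the root control's path and rotation, with body offset location keys closely tracking the root.
# Rig settings such as stretch and local-space switches are mostly held constant.
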LevelSequence /PistolCanary/Anim/Superman/PistolCanary_spa1000c_anm2
bindings [
 Superman_BP {
  id=#1
  possessed_class=PistolCanary_CinematicNPC_Superhero_Main_C
  tracks [
   {type=MovieSceneSkeletalAnimationTrack, sections [{type=MovieSceneSkeletalAnimationSection, end_frame=2820, start_frame=1800}]}
   {
    type=MovieSceneControlRigParameterTrack
    display_name=Fortnite_Biped_CtrlRig
    sections [{
      type=MovieSceneControlRigParameterSection
      channels [
       leg_l_stretch_switch {type=MovieSceneScriptingBoolChannel, default=false, const [false,1021]}
       leg_r_stretch_switch {type=MovieSceneScriptingBoolChannel, default=false, const [false,1021]}
       arm_l_fk_local {type=MovieSceneScriptingBoolChannel, default=true, const [true,26]}
       Auto Interpolate_3 {type=MovieSceneScriptingBoolChannel, default=true, const [true,16]}
       Auto Interpolate {type=MovieSceneScriptingBoolChannel, default=true, const [true,16]}
       arm_l_stretch_switch {type=MovieSceneScriptingBoolChannel, default=false, const [false,1021]}
       Keep Aligned {type=MovieSceneScriptingBoolChannel, default=false, const [false,1021]}
       arm_r_fk_local {type=MovieSceneScriptingBoolChannel, default=true, const [true,30]}
       Auto Interpolate_4 {
        type=MovieSceneScriptingBoolChannel
        default=false
        keys [[2162,false],[2202,true],[2214,true],[2586,true],[2612,false],[2627,false],[2644,false],[2652,false],[2664,true],[2674,false],[2702,false],[2709,false],[2720,false],[2731,false],[2739,true],[2746,true],[2754,true],[2759,false]]
       }
       Auto Interpolate_2 {type=MovieSceneScriptingBoolChannel, default=true, const [true,18]}
       arm_r_stretch_switch {type=MovieSceneScriptingBoolChannel, default=false, const [false,1021]}
       Keep Aligned_2 {type=MovieSceneScriptingBoolChannel, default=false, const [false,1021]}
       head_local {type=MovieSceneScriptingBoolChannel, default=true, const [true,44]}
       Stretch {type=MovieSceneScriptingBoolChannel, default=false, const [false,44]}
       head_stretch {type=MovieSceneScriptingBoolChannel, default=false, const [false,1021]}
       spine_stretch_switch {type=MovieSceneScriptingBoolChannel, default=false, const [false,1021]}
       arm_l_fk_ik_switch {type=MovieSceneScriptingBoolChannel, default=false}
       leg_l_fk_ik_switch {type=MovieSceneScriptingBoolChannel, default=false, const [false,1]}
       arm_r_fk_ik_switch {type=MovieSceneScriptingBoolChannel, default=false}
       leg_r_fk_ik_switch {type=MovieSceneScriptingBoolChannel, default=false, const [false,1]}
       ShowBodyControls {type=MovieSceneScriptingBoolChannel, default=true}
       spine_fk_ik_switch {type=MovieSceneScriptingBoolChannel, default=false}
       neck_fk_ik_switch {type=MovieSceneScriptingBoolChannel, default=false}
       space {type=MovieSceneScriptingByteChannel, default=1, const [1,44]}
       global_ctrl.Location.X {type=MovieSceneScriptingFloatChannel, default=0, const [0,1021]}
       global_ctrl.Location.Y {type=MovieSceneScriptingFloatChannel, default=0, const [0,1021]}
       global_ctrl.Location.Z {type=MovieSceneScriptingFloatChannel, default=0, const [0,1021]}
       global_ctrl.Rotation.X {type=MovieSceneScriptingFloatChannel, default=0, const [0,1021]}
       global_ctrl.Rotation.Y {type=MovieSceneScriptingFloatChannel, default=0, const [0,1021]}
       global_ctrl.Rotation.Z {type=MovieSceneScriptingFloatChannel, default=0, const [0,1021]}
       root_ctrl.Location.X {
        type=MovieSceneScriptingFloatChannel
        default=-2455.8975
        thinned_from=1051
        keys [[1800,0],[1801,-16.9435],[1802,-34.5221],[1803,-52.5658],[1804,-70.9223],[1808,-145.2583],[1809,-163.5844],[1810,-181.5614],[1811,-199.0174],[1812,-215.7609],[1813,-231.6037],[1814,-246.3933],[1815,-260.0497],[1816,-272.6038],[1817,-284.2294],[1819,-306.1966],[1820,-317.6771],[1821,-330.4421],[1822,-345.263],[1835,-552.4824],[1837,-583.0761],[1850,-669.3756],[1853,-684.5209],[1862,-718.155],[1865,-734.9244],[1867,-750.1708],[1869,-769.6822],[1871,-794.312],[1873,-824.8959],[1874,-842.6768],[1876,-883.2992],[1879,-952.2443],[1884,-1081.1327],[1895,-1378.3826],[1899,-1474.1376],[1902,-1536.2792],[1905,-1587.5796],[1934,-1932.6394],[1983,-2274.8325],[2040,-2411.0454],[2041,-2406.8267],[2043,-2383.0251],[2045,-2342.8533],[2047,-2288.646],[2049,-2220.7229],[2051,-2140.6357],[2053,-2050.0916],[2056,-1898.4023],[2060,-1675.4718],[2063,-1499.7329],[2064,-1445.4304],[2065,-1397.4446],[2066,-1358.296],[2069,-1268.3044],[2070,-1241.3571],[2071,-1220.3173],[2072,-1208.0406],[2073,-1207.353],[2074,-1219.34],[2075,-1242.7932],[2076,-1276.3628],[2077,-1318.6981],[2078,-1368.4489],[2079,-1424.2648],[2080,-1484.7959],[2085,-1810.9347],[2086,-1871.4119],[2090,-2084.0413],[2093,-2229.1785],[2095,-2313.5798],[2097,-2384.0349],[2099,-2437.0417],[2100,-2455.8975],[2101,-2469.0603],[2102,-2476.0818],[2107,-2466.4375],[2116,-2396.3005],[2117,-2382.2993],[2118,-2359.738],[2119,-2329.7253],[2120,-2293.3684],[2122,-2206.0491],[2127,-1954.1385],[2131,-1771.3792],[2133,-1686.4148],[2135,-1609.5894],[2136,-1575.2625],[2137,-1544.2133],[2138,-1516.8597],[2139,-1493.6228],[2140,-1474.9255],[2141,-1461.193],[2142,-1452.851],[2156,-1425.7753],[2204,-1423.4679],[2208,-1411.6344],[2209,-1403.6248],[2210,-1389.6248],[2211,-1368.9325],[2212,-1342.6311],[2214,-1284.521],[2215,-1259.3391],[2217,-1200.1761],[2219,-1132.8872],[2221,-1057.9464],[2223,-975.8415],[2225,-887.1104],[2227,-792.2723],[2229,-691.7885],[2231,-586.1141],[2232,-531.4721],[2233,-475.701],[2234,-418.8579],[2235,-360.9991],[2236,-302.1805],[2237,-242.4582],[2238,-181.9061],[2239,-120.5994],[2240,-58.599],[2241,4.0337],[2242,67.2645],[2243,131.0522],[2244,195.3298],[2250,587.3181],[2251,657.7864],[2252,737.012],[2253,823.9235],[2254,917.4541],[2255,1016.5464],[2256,1120.1492],[2258,1336.7206],[2262,1778.3827],[2263,1884.5439],[2264,1986.9349],[2265,2084.5369],[2266,2176.3335],[2267,2261.2859],[2278,3092.1333],[2281,3299.3025],[2283,3424.2954],[2285,3530.0449],[2287,3610.4346],[2289,3666.0212],[2291,3697.4409],[2293,3705.3577],[2295,3690.375],[2297,3653.125],[2299,3594.3318],[2301,3514.7046],[2312,2977.2104],[2322,2448.5522],[2326,2258.4729],[2329,2131.6064],[2332,2018.3512],[2333,1975.5676],[2334,1928.2201],[2335,1876.5247],[2336,1820.6882],[2337,1760.9191],[2338,1697.4247],[2339,1630.4135],[2341,1486.6731],[2343,1331.3628],[2345,1166.1437],[2347,992.6693],[2349,812.5908],[2351,627.5569],[2353,439.1872],[2357,58.9338],[2358,-35.6689],[2359,-129.6762],[2360,-222.8815],[2361,-315.0767],[2362,-406.0547],[2363,-495.6078],[2364,-583.529],[2365,-669.6093],[2366,-753.6441],[2375,-1474.4636],[2380,-1858.7422],[2385,-2221.3994],[2389,-2490.5029],[2392,-2676.8232],[2396,-2898.7356],[2400,-3092.469],[2403,-3215.1741],[2407,-3350.2588],[2415,-3552.8416],[2435,-3949.5649],[2441,-4022.7195],[2445,-4037.3948],[2449,-4015.989],[2452,-3971.854],[2455,-3900.1812],[2458,-3797.8154],[2462,-3614.3027],[2468,-3285.4578],[2480,-2572.7634],[2484,-2357.2234],[2488,-2166.0823],[2489,-2113.2065],[2490,-2055.4509],[2492,-1927.1312],[2494,-1784.7838],[2496,-1632.068],[2503,-1068.6737],[2505,-914.5019],[2506,-840.8257],[2507,-770.0304],[2508,-702.572],[2509,-638.9048],[2510,-579.4833],[2511,-524.7631],[2512,-475.1992],[2513,-430.6043],[2514,-387.4289],[2515,-345.0776],[2516,-303.5825],[2517,-262.9711],[2518,-223.2749],[2519,-184.5214],[2520,-146.7418],[2521,-109.9646],[2522,-74.2209],[2523,-39.5404],[2524,-5.9512],[2525,26.5147],[2526,57.8279],[2527,87.9582],[2528,116.8757],[2529,144.55],[2530,170.9516],[2531,196.0487],[2532,219.8116],[2533,242.2116],[2534,263.2181],[2535,282.8006],[2536,300.9263],[2537,317.5653],[2538,332.6862],[2539,346.2574],[2540,358.2489],[2541,368.6296],[2542,377.3676],[2543,384.4315],[2544,389.7902],[2545,393.4135],[2546,395.269],[2547,395.3265],[2548,393.5557],[2549,389.9243],[2550,384.4009],[2551,376.9565],[2552,367.5592],[2553,356.1782],[2554,342.7838],[2555,327.3441],[2556,309.8828],[2557,290.5024],[2558,269.2889],[2559,246.328],[2560,221.7057],[2561,195.5085],[2562,167.8229],[2563,138.7363],[2564,108.3342],[2565,76.7044],[2566,43.9324],[2567,10.1065],[2568,-24.6878],[2569,-60.3633],[2570,-96.8333],[2571,-134.0111],[2572,-171.8101],[2573,-210.1431],[2575,-288.0653],[2583,-603.7763],[2586,-717.3264],[2588,-788.7463],[2590,-854.7512],[2592,-915.6454],[2594,-972.3912],[2598,-1074.0671],[2599,-1093.6117],[2600,-1110.9275],[2602,-1138.6191],[2605,-1171.798],[2608,-1197.7639],[2611,-1215.5099],[2613,-1221.5443],[2757,-1219.4642],[2759,-1213.3596],[2761,-1201.428],[2763,-1183.3816],[2765,-1159.741],[2767,-1131.0261],[2770,-1079.5785],[2773,-1019.6417],[2777,-929.5562],[2790,-604.4315],[2793,-532.7853],[2795,-487.6775],[2797,-445.3018],[2798,-425.3013],[2799,-415.6262],[2849,-415.6262],[2850,-17203.5215]]
       }
       root_ctrl.Location.Y {
        type=MovieSceneScriptingFloatChannel
        default=128.246
        thinned_from=1051
        keys [[1800,-3.938],[1801,7.1376],[1802,18.7422],[1803,31.5402],[1804,46.0288],[1805,62.5156],[1806,81.1116],[1807,101.7391],[1808,124.1528],[1809,147.9745],[1810,172.7366],[1812,223.0631],[1813,247.6939],[1814,271.4841],[1815,294.2192],[1816,315.8241],[1817,336.3631],[1822,432.5683],[1837,730.0525],[1848,620.4099],[1850,598.9516],[1851,585.4664],[1852,566.9802],[1853,543.8792],[1854,516.5481],[1855,485.3725],[1856,450.737],[1857,413.025],[1858,372.6199],[1859,329.9048],[1860,285.2628],[1861,239.0765],[1862,191.7293],[1863,143.6057],[1865,46.5678],[1866,-1.5734],[1867,-48.9464],[1868,-95.1617],[1869,-139.8292],[1870,-182.5582],[1871,-222.9569],[1872,-260.6335],[1873,-295.1956],[1874,-326.251],[1875,-353.4077],[1876,-378.0477],[1877,-401.8124],[1878,-424.7005],[1880,-467.8533],[1882,-507.5674],[1884,-543.8625],[1886,-576.7493],[1888,-606.2386],[1890,-632.3429],[1892,-655.0752],[1894,-674.4495],[1896,-690.4806],[1898,-703.184],[1900,-712.5758],[1902,-718.6727],[1904,-721.4913],[1906,-721.0485],[1942,-659.1062],[1991,-526.4581],[2001,-488.7169],[2013,-430.8518],[2034,-309.4319],[2039,-278.7839],[2040,-271.6042],[2041,-257.3098],[2042,-235.157],[2043,-205.6592],[2044,-169.7163],[2045,-128.2983],[2046,-82.1023],[2047,-31.2829],[2048,24.066],[2049,83.4864],[2050,146.5184],[2051,212.7026],[2052,281.5804],[2053,352.6915],[2054,425.5778],[2059,800.5619],[2061,946.6697],[2062,1014.6159],[2063,1075.9727],[2064,1128.684],[2065,1170.7408],[2066,1200.0344],[2067,1215.9432],[2068,1227.0508],[2069,1235.1877],[2071,1243.6544],[2072,1244.5211],[2073,1242.2982],[2074,1230.3931],[2075,1208.7084],[2076,1178.3802],[2077,1140.5438],[2078,1096.3359],[2079,1046.8921],[2080,993.3471],[2082,878.5012],[2084,760.8842],[2085,703.8763],[2086,649.584],[2087,599.1423],[2089,503.6805],[2090,457.1495],[2091,411.9384],[2092,368.4359],[2093,327.0285],[2094,288.1006],[2095,252.039],[2096,219.2264],[2097,190.0517],[2098,164.9026],[2099,144.1693],[2100,128.246],[2101,117.528],[2102,112.4114],[2103,112.6445],[2104,114.5342],[2105,117.3632],[2106,121.0481],[2107,125.5012],[2108,130.6399],[2109,136.3768],[2110,142.6277],[2111,149.3079],[2113,163.6142],[2116,186.1615],[2117,197.8772],[2118,217.3248],[2119,243.5134],[2120,275.4542],[2121,312.1592],[2122,352.639],[2123,395.9062],[2126,532.568],[2127,577.1229],[2128,619.5375],[2130,700.734],[2132,778.4169],[2133,815.0854],[2134,849.8532],[2135,882.3808],[2136,912.3286],[2137,939.3545],[2138,963.1159],[2139,983.2656],[2140,999.4558],[2141,1011.3355],[2142,1018.555],[2154,1040.281],[2204,1041.1179],[2207,1048.0762],[2208,1052.8512],[2209,1061.8129],[2210,1076.8156],[2211,1098.5564],[2212,1126.0076],[2214,1186.8367],[2215,1213.5051],[2217,1275.7068],[2219,1346.4276],[2221,1425.139],[2223,1511.3159],[2225,1604.4862],[2228,1756.2456],[2232,1978.0558],[2238,2343.5542],[2252,3268.8491],[2254,3383.7603],[2256,3480.5476],[2258,3555.4211],[2260,3604.5349],[2261,3618.1841],[2262,3623.9004],[2263,3621.1887],[2264,3609.551],[2265,3588.5146],[2266,3557.6001],[2267,3516.2961],[2268,3464.0959],[2270,3340.2566],[2273,3129.5146],[2281,2511.2739],[2284,2290.563],[2285,2208.604],[2286,2119.9712],[2287,2025.3849],[2288,1925.5629],[2289,1821.2349],[2291,1601.9757],[2297,911.7737],[2298,799.6516],[2299,690.2902],[2300,584.4083],[2301,482.7228],[2302,385.2828],[2304,192.5853],[2305,97.3363],[2306,3.2588],[2307,-89.3172],[2308,-180.0565],[2309,-268.6237],[2310,-354.6877],[2311,-437.9189],[2312,-517.9854],[2313,-594.5568],[2314,-667.3019],[2318,-948.5109],[2321,-1152.72],[2323,-1282.8879],[2325,-1406.5352],[2327,-1522.1561],[2329,-1628.2435],[2331,-1723.292],[2332,-1765.8396],[2333,-1804.5558],[2335,-1871.1667],[2337,-1924.4338],[2339,-1965.6431],[2341,-1996.0808],[2343,-2017.0435],[2346,-2033.564],[2350,-2034.6011],[2362,-1991.6039],[2365,-1992.3467],[2376,-2038.7811],[2381,-2038.0458],[2385,-2016.9548],[2388,-1985.9167],[2391,-1939.4972],[2393,-1895.0197],[2395,-1842.2345],[2397,-1781.7566],[2399,-1712.9913],[2400,-1674.6205],[2402,-1587.7124],[2404,-1488.5049],[2406,-1378.4673],[2408,-1259.0275],[2410,-1131.571],[2412,-997.2952],[2414,-857.332],[2416,-713.0193],[2419,-491.1281],[2421,-341.4052],[2422,-265.3714],[2423,-187.2825],[2424,-107.2715],[2425,-25.469],[2426,57.9977],[2427,143.0024],[2428,229.4131],[2429,317.0658],[2430,405.8163],[2432,586.0527],[2442,1505.6222],[2446,1860.9816],[2449,2114.4434],[2452,2352.5833],[2455,2571.7141],[2457,2705.4297],[2476,3831.4167],[2481,4087.0276],[2485,4256.604],[2488,4356.6011],[2491,4420.1724],[2494,4447.5229],[2497,4442.8804],[2501,4394.25],[2506,4279.394],[2515,3979.1182],[2540,2947.8201],[2546,2731.8472],[2551,2581.4404],[2558,2415.7346],[2562,2299.1968],[2567,2124.6694],[2585,1390.8331],[2589,1238.7594],[2592,1134.2737],[2594,1069.3727],[2595,1038.6324],[2596,1010.8956],[2597,985.9184],[2598,963.381],[2600,925.3147],[2601,909.0864],[2602,894.8728],[2603,882.6586],[2605,863.5059],[2607,850.2068],[2609,841.5125],[2612,834.8547],[2739,843.1799],[2754,848.5352],[2756,845.4205],[2757,842.4765],[2759,832.0224],[2760,824.8847],[2761,815.5458],[2762,804.0961],[2763,790.6249],[2764,775.2226],[2765,757.9785],[2766,738.9828],[2767,718.3251],[2768,696.0956],[2769,672.3837],[2770,647.2792],[2772,593.2529],[2774,534.7348],[2776,472.4432],[2778,407.098],[2780,339.4169],[2783,235.0883],[2787,94.516],[2788,59.7099],[2789,25.2182],[2790,-8.8692],[2791,-42.463],[2792,-75.4723],[2793,-107.8082],[2794,-139.3805],[2795,-170.0992],[2796,-199.8746],[2797,-228.6173],[2798,-256.2369],[2799,-269.5978],[2849,-269.5978],[2850,9020.9834]]
       }
       root_ctrl.Location.Z {
        type=MovieSceneScriptingFloatChannel
        default=5696.6084
        thinned_from=1051
        keys [[1800,2.9857],[1801,117.8187],[1802,235.847],[1803,356.8227],[1804,480.4961],[1805,606.6285],[1806,734.9919],[1807,865.3735],[1809,1131.4674],[1812,1541.8433],[1823,3091.6807],[1827,3628.4519],[1830,4008.3843],[1833,4362.5464],[1836,4685.2285],[1839,4972.042],[1842,5216.5703],[1844,5352.895],[1846,5465.8794],[1848,5553.8208],[1891,6519.6333],[1904,6650.3145],[2041,6433.8491],[2057,6718.5542],[2060,6739.6997],[2061,6737.062],[2062,6717.7832],[2064,6637.9497],[2067,6461.874],[2068,6370.6016],[2071,6036.2954],[2072,5947.9912],[2078,5624.6958],[2081,5497.9023],[2084,5411.7915],[2086,5384.0303],[2089,5391.2461],[2092,5439.9146],[2098,5621.0771],[2115,6335.4111],[2123,6558.6562],[2127,6610.1396],[2130,6597.7876],[2134,6505.5127],[2149,6058.9258],[2190,5574.1812],[2205,5530.0845],[2212,5620.3379],[2219,5824.1064],[2249,7126.2109],[2261,7499.9517],[2266,7582.2119],[2270,7587.792],[2282,7407.7251],[2756,7693.3921],[2757,7676.1274],[2758,7546.9551],[2759,7294.4995],[2760,6923.3154],[2761,6437.6821],[2762,5842.2695],[2763,5141.7505],[2764,4340.7949],[2765,3444.0764],[2766,2456.2639],[2767,1382.0315],[2768,226.0471],[2769,-1007.0134],[2770,-2312.4822],[2771,-3685.6892],[2772,-5121.9551],[2773,-6616.6191],[2774,-8165.0015],[2775,-9762.4316],[2776,-11404.2441],[2777,-13085.7676],[2778,-14802.3223],[2780,-18321.8457],[2789,-34660.7422],[2792,-39896.8359],[2795,-44817.6016],[2797,-47860.6484],[2798,-49296.8984],[2799,-49991.6953],[2849,-49991.6953],[2850,8577.6758]]
       }
       root_ctrl.Rotation.X {
        type=MovieSceneScriptingFloatChannel
        default=9.9264
        thinned_from=1051
        keys [[1800,2.0529],[1801,2.1142],[1802,2.2837],[1803,2.5377],[1804,2.8499],[1806,3.5419],[1807,3.8712],[1808,4.16],[1809,4.3914],[1810,4.553],[1811,4.6378],[1812,4.6439],[1813,4.5745],[1814,4.4373],[1815,4.2428],[1816,4.0035],[1817,3.732],[1818,3.4389],[1819,3.1304],[1820,2.8057],[1821,2.4542],[1822,2.0529],[1823,1.441],[1824,0.4942],[1825,-0.7649],[1826,-2.3124],[1827,-4.1226],[1828,-6.1683],[1829,-8.4211],[1830,-10.8521],[1831,-13.4323],[1832,-16.1336],[1833,-18.9286],[1835,-24.6963],[1839,-36.4097],[1850,-70.1997],[1851,-73.4571],[1852,-76.9676],[1853,-80.7064],[1854,-84.6489],[1856,-93.0452],[1859,-106.5461],[1865,-134.3537],[1868,-147.3216],[1870,-155.1666],[1872,-162.1409],[1874,-168.0478],[1876,-172.6907],[1877,-174.477],[1878,-175.8739],[1883,-179.8962],[1884,179.3629],[1897,171.4924],[1905,169.4569],[1944,179.9885],[1945,-179.5945],[1979,-167.6445],[1989,-166.6806],[1991,-167.0844],[1994,-168.9105],[2003,-177.2477],[2006,-178.8624],[2039,-179.094],[2040,-177.3417],[2041,-165.0281],[2042,-144.7579],[2043,-121.0772],[2044,-98.5324],[2045,-81.6707],[2046,-75.0396],[2047,-75.1684],[2048,-75.5426],[2049,-76.1441],[2050,-76.9547],[2051,-77.9562],[2053,-80.4592],[2055,-83.5076],[2059,-90.6594],[2060,-92.5611],[2061,-95.6477],[2062,-105.8639],[2063,-121.5782],[2064,-139.8619],[2065,-157.7863],[2066,-172.4227],[2067,178.1418],[2068,169.7422],[2069,161.7371],[2070,154.3898],[2071,147.9637],[2072,142.7222],[2087,85.4113],[2088,81.2316],[2089,76.3907],[2090,70.9949],[2091,65.1506],[2092,58.9643],[2093,52.5423],[2095,39.4169],[2096,32.9262],[2097,26.6254],[2098,20.6209],[2099,15.0191],[2100,9.9264],[2101,5.4491],[2102,1.6938],[2103,-1.3457],[2116,-39.4606],[2117,-42.5632],[2118,-45.9878],[2119,-49.7052],[2120,-53.6861],[2121,-57.9011],[2122,-62.3212],[2123,-66.9169],[2125,-76.5184],[2128,-91.5062],[2129,-96.8438],[2130,-102.6935],[2131,-108.9306],[2135,-135.2569],[2136,-141.559],[2137,-147.4998],[2138,-152.9545],[2139,-157.7983],[2140,-161.9064],[2141,-165.1541],[2142,-167.4166],[2149,-175.1599],[2153,-178.1707],[2156,-179.2542],[2204,-179.3066],[2205,-176.4925],[2206,-168.8845],[2207,-157.5795],[2208,-143.6135],[2209,-127.9765],[2210,-111.7177],[2211,-95.9369],[2212,-81.7233],[2213,-70.1298],[2214,-62.2016],[2215,-59.0007],[2228,-56.0845],[2230,-55.3521],[2235,-52.8634],[2236,-52.527],[2237,-52.3586],[2240,-52.234],[2241,-52.2907],[2242,-52.9153],[2243,-54.1636],[2244,-55.818],[2246,-59.4934],[2247,-61.1964],[2248,-62.6866],[2249,-63.9312],[2250,-64.9345],[2254,-67.9716],[2255,-68.9918],[2256,-70.3281],[2257,-72.0474],[2258,-74.1441],[2260,-78.8542],[2261,-80.9592],[2262,-82.8081],[2263,-84.3786],[2264,-85.6544],[2265,-86.6497],[2266,-87.3367],[2267,-87.7167],[2285,-87.3639],[2300,-85.3826],[2303,-85.4389],[2306,-86.292],[2312,-88.9219],[2314,-89.2807],[2379,-89.8428],[2446,-85.3334],[2457,-83.4662],[2459,-83.5029],[2465,-84.576],[2523,-81.6739],[2583,-85.3982],[2584,-85.7939],[2585,-86.9508],[2586,-88.8281],[2587,-91.2001],[2588,-94.3946],[2589,-98.3223],[2590,-102.7361],[2591,-107.7377],[2592,-113.4043],[2593,-119.8339],[2595,-134.3227],[2596,-140.446],[2597,-145.7146],[2598,-150.3089],[2599,-154.3876],[2601,-161.5403],[2602,-164.6471],[2603,-167.2757],[2604,-169.4721],[2605,-171.28],[2607,-173.8995],[2609,-175.456],[2613,-176.5515],[2735,-175.9293],[2746,-172.3286],[2754,-172.2721],[2755,-172.675],[2758,-175.7413],[2759,-176.1436],[2849,-176.1436],[2850,-105.4388]]
       }
       root_ctrl.Rotation.Y {
        type=MovieSceneScriptingFloatChannel
        default=-0.4094
        thinned_from=1051
        keys [[1800,0],[1801,-0.0076],[1802,-0.031],[1803,-0.0728],[1804,-0.1364],[1805,-0.2253],[1806,-0.3409],[1807,-0.4827],[1808,-0.6473],[1809,-0.8289],[1811,-1.2078],[1812,-1.3839],[1813,-1.5355],[1814,-1.6509],[1815,-1.7186],[1816,-1.7277],[1817,-1.6676],[1818,-1.5281],[1819,-1.2993],[1820,-0.9724],[1821,-0.5401],[1822,0],[1823,0.5808],[1824,1.1356],[1825,1.662],[1826,2.1578],[1827,2.6218],[1828,3.054],[1829,3.4567],[1830,3.8341],[1831,4.1926],[1833,4.8883],[1834,5.2462],[1835,5.6256],[1836,6.0372],[1837,6.4906],[1838,7.021],[1839,7.6419],[1840,8.3303],[1841,9.0629],[1843,10.568],[1844,11.294],[1845,11.9714],[1846,12.5769],[1847,13.0874],[1848,13.4796],[1849,13.7304],[1850,13.8165],[1851,13.7513],[1852,13.5691],[1853,13.2786],[1854,12.8882],[1855,12.4062],[1856,11.8411],[1857,11.2014],[1858,10.4955],[1859,9.7318],[1860,8.9188],[1861,8.0651],[1862,7.179],[1863,6.2691],[1866,3.4828],[1867,2.564],[1868,1.6645],[1869,0.7932],[1870,-0.0412],[1871,-0.83],[1872,-1.5643],[1873,-2.2354],[1874,-2.8344],[1875,-3.3525],[1876,-3.781],[1877,-4.1109],[1878,-4.3335],[1879,-4.4401],[1880,-4.4771],[1881,-4.4978],[1882,-4.503],[1883,-4.4934],[1884,-4.4699],[1885,-4.4332],[1886,-4.3843],[1887,-4.3238],[1888,-4.2527],[1889,-4.1718],[1890,-4.0818],[1892,-3.8783],[1894,-3.6489],[1897,-3.2714],[1902,-2.6125],[1904,-2.3595],[1905,-2.2389],[1906,-2.1235],[1908,-1.9011],[1910,-1.67],[1912,-1.4304],[1914,-1.184],[1916,-0.9321],[1918,-0.6763],[1923,-0.0291],[1924,0.1002],[1925,0.2289],[1926,0.3569],[1927,0.484],[1928,0.6101],[1930,0.8582],[1932,1.0997],[1934,1.333],[1936,1.5566],[1938,1.7689],[1940,1.9684],[1942,2.1534],[1944,2.3225],[1957,3.3195],[1963,3.7382],[1968,4.0488],[1973,4.3132],[1977,4.4842],[1981,4.6131],[1984,4.6788],[1987,4.7155],[1989,4.7227],[1990,4.6602],[1991,4.4822],[1992,4.2026],[1993,3.8355],[1994,3.3949],[1995,2.8947],[1996,2.349],[1997,1.7717],[1998,1.177],[1999,0.5787],[2000,-0.009],[2001,-0.5722],[2002,-1.0969],[2003,-1.5691],[2004,-1.9748],[2005,-2.2999],[2006,-2.5304],[2007,-2.6524],[2008,-2.6683],[2039,-2.6683],[2040,-2.7203],[2041,-3.0858],[2042,-3.6883],[2043,-4.394],[2044,-5.0716],[2045,-5.5923],[2046,-5.8288],[2053,-6.1768],[2056,-6.2885],[2059,-6.3563],[2060,-6.3667],[2061,-6.3373],[2062,-6.0894],[2063,-5.6342],[2064,-5.0179],[2065,-4.2865],[2066,-3.4858],[2067,-2.662],[2068,-1.861],[2069,-1.1289],[2070,-0.5116],[2071,-0.0552],[2072,0.1944],[2073,0.2253],[2074,0.2148],[2075,0.1964],[2076,0.1713],[2077,0.1408],[2078,0.106],[2079,0.0682],[2080,0.0285],[2081,-0.0118],[2082,-0.0516],[2083,-0.0896],[2084,-0.1246],[2085,-0.1555],[2086,-0.181],[2087,-0.1999],[2091,-0.2619],[2097,-0.3591],[2155,-1.3797],[2174,-1.6602],[2189,-1.8299],[2202,-1.9281],[2203,-1.9451],[2204,-1.9788],[2205,-1.8648],[2206,-1.2045],[2207,0.1115],[2208,1.6622],[2209,2.7418],[2210,2.9893],[2211,2.3754],[2212,1.2198],[2213,-0.0375],[2214,-1.0679],[2215,-1.7302],[2216,-1.972],[2218,-1.9717],[2219,-1.9662],[2220,-1.9556],[2221,-1.9405],[2223,-1.8988],[2224,-1.5308],[2225,-0.5466],[2226,0.9474],[2227,2.8452],[2228,5.0412],[2229,7.4302],[2231,12.3686],[2232,14.7129],[2233,16.8408],[2234,18.6564],[2235,20.0662],[2236,20.9791],[2237,21.3035],[2240,21.2872],[2241,21.2453],[2242,20.3742],[2243,18.2272],[2244,14.9356],[2245,10.6391],[2246,5.4964],[2247,-0.3094],[2248,-6.5715],[2250,-19.5541],[2251,-25.7661],[2252,-31.5343],[2253,-36.7731],[2254,-41.3931],[2255,-45.2996],[2256,-48.3911],[2257,-50.5581],[2258,-51.6851],[2259,-51.6553],[2260,-50.9602],[2261,-50.1213],[2262,-49.0723],[2263,-47.7485],[2264,-46.0867],[2265,-43.0506],[2266,-37.9882],[2267,-31.4126],[2268,-23.8336],[2269,-15.6718],[2270,-7.2604],[2271,1.0234],[2272,8.8029],[2273,15.7011],[2274,21.3409],[2275,25.3457],[2276,27.3383],[2277,26.9419],[2278,24.772],[2279,21.8974],[2280,18.6212],[2281,15.2467],[2282,12.0768],[2283,9.4147],[2284,7.3363],[2285,5.8761],[2286,5.2754],[2287,5.1904],[2289,5.0514],[2291,4.9463],[2293,4.8667],[2294,4.4365],[2295,3.2802],[2296,1.4963],[2297,-0.8168],[2298,-3.5601],[2299,-6.6347],[2300,-9.9415],[2301,-13.3814],[2302,-16.8565],[2303,-20.2758],[2304,-23.5427],[2305,-26.5587],[2306,-29.2248],[2307,-31.4417],[2309,-35.275],[2310,-37.035],[2311,-38.6087],[2312,-39.9353],[2313,-40.9544],[2314,-41.6059],[2315,-41.8273],[2316,-41.3926],[2317,-40.175],[2318,-38.264],[2319,-35.7543],[2320,-32.745],[2321,-29.3387],[2322,-25.6403],[2323,-21.7559],[2325,-13.8528],[2326,-10.0422],[2327,-6.4597],[2328,-3.2009],[2329,-0.3567],[2330,1.9876],[2331,3.753],[2332,4.8668],[2333,5.263],[2334,5.1445],[2335,4.7662],[2336,4.1494],[2337,3.3209],[2338,2.3134],[2339,1.1657],[2340,-0.0783],[2341,-1.3697],[2342,-2.6563],[2343,-3.8837],[2344,-4.997],[2345,-5.9427],[2346,-6.6704],[2347,-7.1341],[2348,-7.2935],[2349,-6.9081],[2350,-5.8147],[2351,-4.0939],[2352,-1.8263],[2353,0.9075],[2354,4.0268],[2355,7.4513],[2356,11.1003],[2357,14.8934],[2359,22.5899],[2360,26.3323],[2361,29.897],[2362,33.2035],[2363,36.1714],[2364,38.7203],[2365,40.7698],[2368,45.9877],[2369,47.614],[2370,49.1421],[2371,50.5475],[2372,51.8055],[2373,52.8916],[2374,53.7813],[2375,54.45],[2376,54.8732],[2377,55.0263],[2378,54.8383],[2379,54.2683],[2380,53.3363],[2381,52.0626],[2382,50.4673],[2383,48.5707],[2384,46.3934],[2385,43.9553],[2386,41.2747],[2387,38.3727],[2388,35.2707],[2389,31.9878],[2390,28.5428],[2391,24.9549],[2392,21.2351],[2393,17.3983],[2394,13.4693],[2395,9.4727],[2396,5.4331],[2397,1.3752],[2398,-2.6765],[2399,-6.6973],[2400,-10.6652],[2401,-14.5619],[2402,-18.369],[2403,-22.0678],[2404,-25.6396],[2405,-29.0659],[2406,-32.328],[2407,-35.4073],[2408,-38.2852],[2409,-40.9404],[2410,-43.3505],[2411,-45.4952],[2412,-47.3542],[2415,-52.3765],[2417,-55.411],[2418,-56.7636],[2419,-57.9752],[2420,-59.0221],[2421,-59.881],[2422,-60.5355],[2423,-60.9649],[2424,-61.1392],[2427,-61.0647],[2438,-59.62],[2442,-59.5534],[2453,-61.8048],[2456,-62.0524],[2457,-61.8631],[2458,-61.2294],[2459,-60.0369],[2460,-58.1753],[2461,-55.6135],[2462,-52.5113],[2463,-49.2244],[2464,-46.1101],[2465,-43.3626],[2466,-41.0268],[2467,-39.075],[2468,-37.4579],[2469,-36.1262],[2470,-35.037],[2471,-34.1545],[2472,-33.4498],[2473,-32.8992],[2476,-31.5008],[2478,-30.6949],[2480,-30.0789],[2481,-29.8658],[2482,-29.7288],[2483,-29.6772],[2484,-29.7611],[2485,-30.0149],[2486,-30.4247],[2487,-30.9767],[2488,-31.6564],[2489,-32.4493],[2490,-33.3411],[2491,-34.3178],[2493,-36.4696],[2499,-43.498],[2501,-45.6586],[2503,-47.5431],[2504,-48.3476],[2508,-50.8565],[2512,-52.9476],[2516,-54.5831],[2519,-55.4575],[2522,-55.9867],[2525,-56.1299],[2528,-55.8445],[2530,-55.3939],[2532,-54.6767],[2533,-54.1311],[2534,-53.4562],[2535,-52.6599],[2536,-51.7501],[2538,-49.6212],[2540,-47.1324],[2542,-44.3468],[2544,-41.3272],[2548,-34.8394],[2553,-26.5421],[2555,-23.3695],[2556,-21.8462],[2557,-20.3752],[2558,-18.9644],[2559,-17.622],[2560,-16.3558],[2561,-15.174],[2562,-14.0731],[2563,-13.0394],[2564,-12.0659],[2565,-11.1454],[2566,-10.2709],[2567,-9.4354],[2568,-8.6316],[2569,-7.8527],[2573,-4.8426],[2574,-4.0811],[2575,-3.3019],[2576,-2.4979],[2577,-1.6619],[2578,-0.787],[2579,0.134],[2580,1.1082],[2581,2.1426],[2582,3.2444],[2583,4.4348],[2584,5.6864],[2585,6.9556],[2586,6.9857],[2587,5.5914],[2588,4.3089],[2589,3.5083],[2590,3.1579],[2591,3.2845],[2592,3.3473],[2593,3.385],[2594,3.4061],[2595,3.5134],[2596,3.7762],[2597,4.1118],[2598,4.4144],[2599,4.556],[2600,4.3888],[2601,3.7488],[2602,2.8823],[2603,2.1571],[2604,1.5769],[2605,1.1395],[2606,0.8379],[2607,0.6609],[2608,0.5935],[2609,0.6169],[2610,0.7091],[2611,0.845],[2612,0.997],[2613,1.1352],[2614,1.2281],[2615,1.2593],[2616,1.2339],[2617,1.1602],[2618,1.0471],[2619,0.9037],[2620,0.7388],[2621,0.5614],[2622,0.3804],[2623,0.2048],[2624,0.0433],[2625,-0.0949],[2626,-0.201],[2627,-0.2661],[2628,-0.3076],[2629,-0.3484],[2630,-0.3879],[2631,-0.4252],[2632,-0.4598],[2633,-0.491],[2634,-0.5181],[2635,-0.5403],[2636,-0.557],[2637,-0.5676],[2638,-0.5712],[2646,-0.5712],[2647,-0.5765],[2648,-0.5931],[2649,-0.6226],[2650,-0.6662],[2651,-0.7255],[2652,-0.8014],[2653,-0.8944],[2654,-1.0033],[2655,-1.1254],[2657,-1.3873],[2658,-1.5136],[2659,-1.6292],[2660,-1.7308],[2661,-1.8174],[2662,-1.8896],[2663,-1.9486],[2664,-1.9961],[2665,-2.0339],[2666,-2.0632],[2667,-2.0854],[2668,-2.1014],[2669,-2.1121],[2671,-2.1201],[2702,-2.1145],[2703,-2.1109],[2704,-2.1013],[2705,-2.0865],[2707,-2.0441],[2713,-1.8772],[2714,-1.8548],[2715,-1.8372],[2716,-1.8257],[2717,-1.8216],[2733,-1.8216],[2734,-1.8713],[2735,-2.0104],[2736,-2.2241],[2737,-2.4975],[2738,-2.8156],[2739,-3.1635],[2741,-3.8891],[2742,-4.2369],[2743,-4.555],[2744,-4.8284],[2745,-5.0421],[2746,-5.1812],[2747,-5.2309],[2754,-5.2309],[2755,-4.8764],[2756,-4.0308],[2757,-3.0217],[2758,-2.1762],[2759,-1.8216],[2849,-1.8216],[2850,88.6399]]
       }
       root_ctrl.Rotation.Z {
        type=MovieSceneScriptingFloatChannel
        default=131.5375
        thinned_from=1051
        keys [[1800,0],[1801,0.1897],[1802,0.7233],[1803,1.5475],[1804,2.6097],[1805,3.8569],[1806,5.2358],[1807,6.6923],[1808,8.1718],[1809,9.6193],[1810,10.9789],[1811,12.1946],[1812,13.2098],[1813,13.968],[1814,14.4127],[1815,14.4875],[1816,14.1364],[1817,13.3037],[1818,11.9345],[1819,9.9739],[1820,7.3677],[1821,4.0615],[1822,0],[1823,-4.9055],[1824,-10.5977],[1825,-16.9137],[1826,-23.6906],[1827,-30.7651],[1829,-45.1534],[1830,-52.1398],[1831,-58.7695],[1832,-64.8783],[1833,-70.3024],[1834,-74.8781],[1835,-78.4414],[1836,-80.8284],[1837,-81.8752],[1850,-86.0911],[1851,-86.511],[1852,-87.1285],[1853,-87.933],[1855,-90.0603],[1857,-92.8074],[1859,-96.0888],[1862,-101.8262],[1876,-133.652],[1891,-162.0682],[1899,-174.4706],[1903,-179.3798],[1904,179.5589],[1939,152.3456],[1970,136.6822],[1984,133.1749],[2039,133.0275],[2040,133.2502],[2041,134.8156],[2042,137.3934],[2044,143.2766],[2045,145.4242],[2046,146.2692],[2052,145.0821],[2078,133.1315],[2085,131.5506],[2202,134.4397],[2203,133.9793],[2204,132.8055],[2206,129.8009],[2207,128.9371],[2208,129.2207],[2209,130.5792],[2211,133.9297],[2212,134.8924],[2213,135.171],[2216,134.5597],[2223,134.3783],[2224,134.5156],[2225,135.0181],[2227,136.8941],[2233,145.4519],[2235,147.5957],[2236,148.1765],[2241,147.7252],[2242,146.9607],[2243,145.3242],[2244,143.0424],[2249,128.6847],[2250,125.9783],[2251,122.8086],[2252,118.6908],[2253,113.8537],[2254,108.5416],[2256,97.4825],[2257,92.1696],[2258,87.1814],[2261,73.3935],[2262,68.9132],[2263,64.625],[2264,60.6274],[2265,57.0574],[2266,53.9922],[2267,51.5463],[2268,49.912],[2283,32.0246],[2284,30.5543],[2285,28.5623],[2286,26.1148],[2287,23.2738],[2288,20.0968],[2289,16.6464],[2290,12.9855],[2291,9.1768],[2292,5.2832],[2293,1.3673],[2294,-2.5218],[2295,-6.3364],[2296,-10.0142],[2297,-13.4904],[2298,-16.6985],[2299,-19.5706],[2300,-22.0371],[2301,-24.0273],[2302,-25.5377],[2303,-26.924],[2304,-28.2267],[2305,-29.425],[2306,-30.5055],[2307,-31.4643],[2308,-32.3119],[2309,-33.0581],[2310,-33.7102],[2312,-34.7919],[2317,-37.1935],[2332,-45.4986],[2333,-46.2228],[2334,-47.0812],[2335,-48.0695],[2336,-49.1789],[2337,-50.4008],[2339,-53.1484],[2341,-56.2455],[2344,-61.4053],[2356,-84.4121],[2359,-89.5963],[2361,-92.636],[2363,-95.2341],[2365,-97.2812],[2367,-98.7125],[2389,-109.8033],[2391,-111.3148],[2392,-112.3575],[2393,-114.0319],[2394,-116.5334],[2395,-119.5792],[2397,-126.1792],[2398,-129.173],[2399,-131.5891],[2402,-137.162],[2421,-166.9309],[2422,-168.874],[2425,-176.514],[2426,-178.6983],[2427,179.7843],[2441,171.7652],[2442,170.9786],[2443,169.7757],[2445,166.3117],[2447,161.749],[2454,142.6299],[2456,137.8857],[2457,135.9111],[2458,134.3142],[2459,133.1568],[2469,126.2623],[2487,110.4679],[2489,108.1864],[2491,105.1378],[2493,101.395],[2495,97.0313],[2497,92.1219],[2499,86.7443],[2502,77.9796],[2506,65.4469],[2515,36.1674],[2518,26.6298],[2520,20.4214],[2521,17.3742],[2522,14.3701],[2523,11.4127],[2524,8.5056],[2525,5.6521],[2526,2.8577],[2527,0.1185],[2528,-2.5554],[2529,-5.1636],[2530,-7.7035],[2531,-10.1723],[2532,-12.5606],[2533,-14.8608],[2534,-17.0736],[2535,-19.1997],[2536,-21.2391],[2537,-23.1918],[2538,-25.0575],[2539,-26.8349],[2540,-28.5227],[2541,-30.1191],[2542,-31.6218],[2543,-33.0283],[2544,-34.3359],[2545,-35.5414],[2546,-36.6414],[2547,-37.6327],[2548,-38.5117],[2549,-39.2747],[2550,-39.918],[2551,-40.4381],[2552,-40.8309],[2553,-41.0929],[2554,-41.2202],[2583,-39.9131],[2584,-39.4133],[2585,-38.1104],[2586,-36.1332],[2587,-33.6008],[2588,-30.5726],[2589,-27.4158],[2592,-17.0154],[2593,-13.6621],[2594,-9.9206],[2595,-7.3097],[2596,-3.9269],[2597,-0.3951],[2598,3.2187],[2599,6.6796],[2600,9.9024],[2601,12.7411],[2602,15.1761],[2603,17.2842],[2604,19.0718],[2605,20.5517],[2606,21.7418],[2607,22.6647],[2608,23.3461],[2609,23.8144],[2610,24.0995],[2611,24.2323],[2612,24.2439],[2613,24.1654],[2615,23.9294],[2646,23.4898],[2647,23.6103],[2648,23.9921],[2649,24.6687],[2650,25.6765],[2651,27.0524],[2652,28.8299],[2653,31.0304],[2654,33.6511],[2655,36.6502],[2656,39.9359],[2658,46.7981],[2659,50.0694],[2660,53.0777],[2661,55.7609],[2662,58.0963],[2663,60.0874],[2664,61.7526],[2665,63.1171],[2666,64.208],[2667,65.0519],[2668,65.6732],[2669,66.094],[2670,66.334],[2702,66.1872],[2703,66.0466],[2704,65.6706],[2705,65.0956],[2707,63.4939],[2712,58.5492],[2713,57.6896],[2714,56.9579],[2715,56.3902],[2716,56.023],[2717,55.8926],[2849,55.8926],[2850,125.91]]
       }
       root_ctrl.Scale.X {type=MovieSceneScriptingFloatChannel, default=1, const [1,1021]}
       root_ctrl.Scale.Y {type=MovieSceneScriptingFloatChannel, default=1, const [1,1021]}
       root_ctrl.Scale.Z {type=MovieSceneScriptingFloatChannel, default=1, const [1,1021]}
       body_offset_ctrl.Location.X {
        type=MovieSceneScriptingFloatChannel
        default=0
        thinned_from=997
        keys [[1800,-0.0001],[1801,-16.9395],[1802,-34.5068],[1803,-52.5342],[1804,-70.8699],[1808,-145.0449],[1809,-163.2896],[1810,-181.1589],[1811,-198.4783],[1812,-215.0574],[1813,-230.7139],[1814,-245.3063],[1815,-258.7717],[1816,-271.1639],[1819,-304.748],[1820,-316.5015],[1821,-329.7382],[1822,-345.2631],[1825,-397.1631],[1827,-434.1639],[1830,-493.9089],[1837,-635.8002],[1843,-701.6097],[1848,-748.4297],[1851,-771.8945],[1853,-783.9662],[1856,-796.2003],[1859,-802.2159],[1866,-805.0736],[1868,-808.833],[1870,-817.1705],[1871,-823.6576],[1872,-831.995],[1873,-842.3933],[1874,-855.0454],[1875,-870.1245],[1877,-906.3486],[1879,-949.281],[1883,-1047.335],[1896,-1395.0754],[1900,-1489.666],[1903,-1550.2114],[1906,-1599.0132],[1932,-1917.8452],[1983,-2264.5322],[2039,-2406.687],[2040,-2405.0376],[2041,-2385.3223],[2043,-2325.3218],[2046,-2265.5574],[2048,-2203.989],[2050,-2128.6877],[2052,-2041.5236],[2054,-1944.3889],[2057,-1784.152],[2062,-1493.7504],[2063,-1440.5502],[2064,-1398.3154],[2065,-1367.6057],[2066,-1345.5181],[2070,-1271.4545],[2071,-1258.0374],[2072,-1251.5706],[2073,-1255.0063],[2074,-1270.7211],[2075,-1297.707],[2076,-1334.5934],[2077,-1380.009],[2078,-1432.5833],[2079,-1490.9462],[2081,-1619.5627],[2084,-1821.718],[2086,-1946.7408],[2089,-2105.6223],[2091,-2202.5396],[2093,-2288.7881],[2095,-2361.1843],[2097,-2417.5557],[2099,-2456.3401],[2100,-2468.6768],[2101,-2476.0283],[2102,-2478.1465],[2107,-2449.385],[2116,-2348.7908],[2117,-2331.7595],[2118,-2306.0239],[2119,-2272.781],[2120,-2233.2292],[2122,-2139.9912],[2127,-1879.7401],[2130,-1743.1459],[2132,-1660.7032],[2134,-1588.47],[2135,-1556.9167],[2136,-1528.6621],[2137,-1503.8455],[2138,-1482.5897],[2139,-1465.0219],[2140,-1451.2871],[2141,-1441.5581],[2142,-1436.0364],[2204,-1420.8779],[2205,-1414.1276],[2215,-1198.8149],[2217,-1140.0597],[2219,-1073.0483],[2221,-998.3174],[2223,-916.3608],[2225,-827.6697],[2227,-732.7344],[2228,-683.0791],[2229,-632.0469],[2230,-579.6987],[2231,-526.0967],[2232,-471.3018],[2233,-415.3745],[2234,-358.377],[2235,-300.3706],[2236,-241.416],[2237,-181.5742],[2238,-120.9072],[2239,-59.4756],[2240,2.6592],[2241,65.4336],[2242,128.7842],[2243,192.6885],[2244,257.1299],[2246,387.5273],[2250,653.123],[2251,725.6201],[2252,808.3203],[2253,899.7568],[2254,998.3984],[2255,1102.6953],[2256,1211.1191],[2260,1658.5557],[2262,1874.752],[2263,1977.8818],[2264,2076.7471],[2265,2170.6729],[2266,2258.9883],[2267,2341.0049],[2278,3153.0146],[2281,3355.3838],[2283,3477.1074],[2284,3531.0791],[2286,3617.2793],[2288,3675.4375],[2290,3706.8516],[2292,3713.0605],[2294,3695.7998],[2296,3656.8682],[2298,3597.9658],[2300,3520.5957],[2312,2920.355],[2322,2384.1621],[2326,2190.989],[2329,2061.9248],[2332,1946.5819],[2333,1902.9103],[2334,1854.5311],[2335,1801.6707],[2336,1744.5543],[2337,1683.4092],[2338,1618.4615],[2339,1549.9387],[2340,1478.0676],[2341,1403.074],[2342,1325.1846],[2344,1161.6147],[2346,989.1448],[2348,809.5356],[2350,624.5195],[2352,435.7959],[2356,53.8418],[2357,-41.4077],[2358,-136.1646],[2359,-230.2344],[2360,-323.4233],[2361,-415.5376],[2362,-506.3853],[2363,-595.7744],[2364,-683.5137],[2365,-769.4097],[2366,-853.271],[2375,-1572.6064],[2380,-1955.7881],[2385,-2317.1494],[2389,-2584.9365],[2392,-2769.6436],[2395,-2933.2988],[2400,-3165.5254],[2404,-3316.3359],[2409,-3462.1533],[2436,-3960.5144],[2441,-4013.7136],[2445,-4019.573],[2448,-3995.8757],[2451,-3947.4219],[2454,-3873.4443],[2457,-3772.6382],[2460,-3643.2102],[2465,-3382.1279],[2480,-2485.2705],[2484,-2266.8955],[2488,-2073.2256],[2489,-2019.6104],[2490,-1961.0459],[2491,-1898.0156],[2493,-1760.499],[2495,-1610.9482],[2502,-1048.8154],[2504,-892.9951],[2505,-818.2432],[2506,-746.2109],[2507,-677.3555],[2508,-612.1299],[2509,-550.9814],[2510,-494.3535],[2511,-442.6865],[2512,-396.417],[2513,-355.3389],[2514,-315.9014],[2515,-277.4961],[2516,-240.1377],[2517,-203.835],[2518,-168.5996],[2519,-134.4385],[2520,-101.3613],[2521,-69.375],[2522,-38.4883],[2523,-8.709],[2524,19.957],[2525,47.499],[2526,73.9082],[2527,99.1758],[2528,123.29],[2529,146.2393],[2530,168.0107],[2531,188.5889],[2532,207.957],[2533,226.0986],[2534,242.9951],[2535,258.626],[2536,272.9678],[2537,285.998],[2538,297.6914],[2539,308.0215],[2540,316.9619],[2541,324.4844],[2542,330.5586],[2543,335.1543],[2544,338.2402],[2545,339.7861],[2546,339.7588],[2547,338.127],[2548,334.8594],[2549,329.9229],[2550,323.2852],[2551,314.917],[2552,304.7871],[2553,292.8652],[2554,279.124],[2555,263.5342],[2556,246.0229],[2557,226.5938],[2558,205.333],[2559,182.3267],[2560,157.6611],[2561,131.4233],[2562,103.6997],[2563,74.5771],[2564,44.1411],[2565,12.479],[2566,-20.3237],[2567,-54.1792],[2568,-89.002],[2569,-124.7051],[2570,-161.2021],[2571,-198.4067],[2572,-236.2324],[2574,-313.4004],[2583,-668.5598],[2584,-706.4333],[2585,-742.5188],[2586,-776.7332],[2587,-808.9717],[2588,-839.1505],[2589,-867.2467],[2590,-893.3295],[2592,-940.2689],[2598,-1064.4922],[2614,-1215.8179],[2757,-1215.5916],[2759,-1210.2952],[2761,-1198.3635],[2763,-1180.3171],[2765,-1156.6765],[2767,-1127.9617],[2770,-1076.514],[2773,-1016.5773],[2777,-926.4918],[2790,-601.3671],[2793,-529.7209],[2795,-484.6131],[2797,-442.2375],[2798,-422.2369],[2799,-412.5618],[2820,-412.5618]]
       }
       body_offset_ctrl.Location.Y {
        type=MovieSceneScriptingFloatChannel
        default=0
        thinned_from=997
        keys [[1800,-1.2381],[1801,9.945],[1802,21.8475],[1803,35.0927],[1804,50.1333],[1805,67.2316],[1806,86.455],[1807,107.6856],[1808,130.6432],[1809,154.9208],[1812,230.6664],[1813,255.2522],[1814,278.8667],[1815,301.3033],[1816,322.4977],[1817,342.5282],[1819,380.0277],[1828,546.2724],[1830,585.2907],[1836,712.4584],[1837,732.363],[1845,655.5925],[1850,600.3945],[1851,586.1725],[1852,567.1696],[1853,543.8176],[1854,516.5395],[1855,485.7482],[1856,451.8415],[1857,415.1996],[1858,376.1855],[1859,335.143],[1860,292.3991],[1861,248.2648],[1862,203.0402],[1863,157.0185],[1866,17.1044],[1867,-29.1292],[1868,-74.6114],[1869,-118.9813],[1870,-161.8543],[1871,-202.8214],[1872,-241.4505],[1873,-277.289],[1874,-309.8668],[1875,-338.7003],[1876,-365.0705],[1877,-390.5119],[1878,-414.9144],[1879,-438.1689],[1881,-481.9126],[1883,-522.512],[1885,-559.8865],[1887,-593.9518],[1889,-624.6259],[1891,-651.8321],[1893,-675.5027],[1895,-695.5823],[1897,-712.0274],[1899,-724.8089],[1901,-733.9095],[1903,-739.3235],[1905,-741.0538],[1932,-688.2156],[1991,-506.1375],[2002,-482.1659],[2005,-472.556],[2014,-428.4906],[2039,-281.7372],[2040,-272.5003],[2041,-243.3329],[2042,-196.5916],[2043,-142.2245],[2044,-90.952],[2045,-45.5116],[2046,0.3004],[2047,51.133],[2048,106.4919],[2049,165.9108],[2050,228.9176],[2051,295.0387],[2052,363.8],[2053,434.7257],[2054,507.3434],[2059,879.4059],[2060,952.3845],[2061,1023.2844],[2062,1086.7726],[2063,1137.4786],[2064,1172.7179],[2065,1193.9967],[2066,1205.1754],[2067,1209.7218],[2069,1210.3815],[2072,1200.901],[2073,1195.1818],[2074,1180.1823],[2075,1155.6598],[2076,1122.7579],[2077,1082.6173],[2078,1036.3781],[2079,985.1769],[2080,930.1481],[2084,694.4172],[2085,637.2316],[2086,583.0036],[2087,532.8605],[2088,485.4894],[2089,439.195],[2090,394.5085],[2091,351.9235],[2092,311.883],[2093,274.7692],[2094,240.9005],[2095,210.5397],[2096,183.8957],[2097,161.1476],[2098,142.4538],[2099,127.9714],[2100,117.8717],[2101,112.3483],[2102,111.6203],[2103,115.4152],[2104,120.736],[2105,126.9821],[2106,134.0612],[2107,141.8766],[2108,150.3371],[2109,159.3463],[2110,168.8117],[2112,188.737],[2116,229.9518],[2117,244.3937],[2118,266.6891],[2119,295.7672],[2120,330.5563],[2121,369.9855],[2122,412.984],[2123,458.4875],[2126,599.5095],[2127,644.5724],[2128,686.9758],[2130,766.2682],[2131,803.5558],[2132,838.7243],[2133,871.4567],[2134,901.5134],[2135,928.7214],[2136,952.9611],[2137,974.1462],[2138,992.2053],[2139,1007.0607],[2140,1018.6183],[2141,1026.7579],[2142,1031.3351],[2204,1038.7838],[2205,1044.4583],[2215,1275.5465],[2217,1337.4156],[2219,1407.8566],[2221,1486.2291],[2223,1572.0133],[2226,1713.4674],[2229,1868.7194],[2233,2094.1985],[2240,2528.1086],[2251,3270.4519],[2253,3386.864],[2255,3480.2371],[2257,3547.8503],[2259,3588.2507],[2261,3600.7261],[2262,3596.209],[2263,3584.252],[2264,3564.5554],[2265,3536.7468],[2266,3500.3997],[2267,3455.0559],[2268,3400.2612],[2270,3272.7219],[2273,3056.6855],[2281,2427.8657],[2284,2203.7878],[2285,2120.1597],[2286,2029.5861],[2287,1932.9371],[2288,1831.0807],[2289,1724.8756],[2291,1502.7682],[2296,927.1555],[2297,814.1896],[2298,703.4387],[2299,595.5002],[2300,490.9669],[2301,390.4308],[2302,293.8752],[2303,198.1261],[2304,102.9074],[2305,8.5451],[2306,-84.6366],[2307,-176.3133],[2308,-266.1639],[2309,-353.869],[2310,-439.1092],[2311,-521.5682],[2312,-600.9276],[2313,-676.8709],[2314,-749.0809],[2318,-1028.1932],[2321,-1230.6005],[2323,-1359.4725],[2325,-1481.7714],[2327,-1596.0204],[2329,-1700.744],[2331,-1794.4667],[2332,-1836.2743],[2333,-1874.0741],[2335,-1938.3143],[2337,-1988.4979],[2339,-2025.9413],[2341,-2051.9675],[2343,-2067.9207],[2346,-2075.9509],[2350,-2064.3083],[2361,-1989.0487],[2364,-1980.1119],[2367,-1985.1764],[2376,-2016.0028],[2381,-2010.8213],[2384,-1994.4105],[2387,-1965.2794],[2390,-1921.183],[2391,-1902.4757],[2393,-1853.7152],[2395,-1792.4076],[2398,-1684.9945],[2400,-1605.5326],[2402,-1514.3182],[2404,-1411.4457],[2406,-1298.1747],[2408,-1175.764],[2410,-1045.4716],[2412,-908.5545],[2414,-766.2703],[2416,-619.8748],[2419,-395.3192],[2421,-244.1366],[2422,-167.3377],[2423,-88.4256],[2424,-7.7039],[2425,74.6017],[2426,158.3527],[2427,243.4757],[2428,329.9181],[2429,417.5871],[2431,596.0549],[2443,1695.674],[2446,1959.1896],[2449,2209.1838],[2452,2441.7952],[2455,2654.2366],[2458,2844.6799],[2476,3885.7581],[2481,4134.6777],[2485,4298.752],[2488,4394.418],[2490,4436.9766],[2493,4467.5898],[2496,4462.3184],[2499,4425.7842],[2503,4336.7588],[2509,4140.9658],[2538,2932.2517],[2545,2680.6677],[2550,2529.3918],[2554,2430.0378],[2558,2338.449],[2562,2222.0559],[2566,2084.6555],[2573,1810.2169],[2588,1189.1842],[2590,1112.8654],[2591,1077.9161],[2592,1045.6759],[2593,1016.4323],[2594,990.319],[2595,967.3058],[2596,947.2087],[2597,929.7175],[2598,914.4474],[2614,833.7194],[2757,833.4069],[2758,830.9406],[2759,826.093],[2760,818.9553],[2761,809.6164],[2762,798.1667],[2763,784.6955],[2764,769.2932],[2765,752.049],[2766,733.0534],[2767,712.3957],[2768,690.1662],[2769,666.4543],[2770,641.3498],[2772,587.3234],[2774,528.8054],[2776,466.5138],[2778,401.1686],[2780,333.4875],[2783,229.1588],[2787,88.5866],[2788,53.7804],[2789,19.2887],[2790,-14.7987],[2791,-48.3924],[2792,-81.4017],[2793,-113.7376],[2794,-145.3099],[2795,-176.0287],[2796,-205.804],[2797,-234.5467],[2798,-262.1664],[2799,-275.5272],[2820,-275.5272]]
       }
       body_offset_ctrl.Location.Z {
        type=MovieSceneScriptingFloatChannel
        default=0
        thinned_from=991
        keys [[1800,103.5105],[1801,218.3406],[1802,336.3601],[1803,457.321],[1804,580.9734],[1805,707.0789],[1806,835.4109],[1808,1097.9363],[1810,1367.1785],[1814,1921.7976],[1823,3192.2234],[1827,3728.5789],[1830,4106.7524],[1833,4457.0278],[1836,4773.4165],[1839,5051.7197],[1841,5212.918],[1843,5352.4053],[1845,5468.6255],[1847,5560.041],[1849,5625.1055],[1896,6489.8618],[2040,6330.2822],[2041,6336.6138],[2042,6358.1558],[2045,6496.2729],[2056,6712.623],[2060,6734.3379],[2061,6726.3335],[2062,6689.5869],[2067,6361.5078],[2068,6271.8647],[2071,5951.5308],[2072,5868.5244],[2077,5618.7817],[2080,5499.3818],[2083,5417.3843],[2085,5390.1689],[2087,5389.5742],[2089,5415.7847],[2093,5525.3457],[2114,6382.6123],[2121,6565.6924],[2125,6614.7192],[2128,6608.7061],[2130,6574.814],[2142,6140.271],[2154,5854.248],[2194,5452],[2205,5429.7139],[2221,5951.0811],[2246,7062.6152],[2260,7487.6333],[2266,7585.1836],[2270,7590.6772],[2282,7411.0635],[2590,7632.8423],[2598,7603.6079],[2756,7590.7446],[2757,7574.4922],[2758,7446.2222],[2759,7194.1602],[2760,6822.9761],[2761,6337.3428],[2762,5741.9302],[2763,5041.4111],[2764,4240.4556],[2765,3343.7371],[2766,2355.9246],[2767,1281.6921],[2768,125.7077],[2769,-1107.3528],[2770,-2412.8215],[2771,-3786.0286],[2772,-5222.2944],[2773,-6716.9585],[2774,-8265.3408],[2775,-9862.7705],[2776,-11504.583],[2777,-13186.1064],[2778,-14902.6611],[2780,-18422.1855],[2789,-34761.082],[2792,-39997.1758],[2795,-44917.9414],[2797,-47960.9883],[2798,-49397.2383],[2799,-50092.0352],[2820,-50092.0352]]
       }
       body_offset_ctrl.Rotation.X {
        type=MovieSceneScriptingFloatChannel
        default=0
        thinned_from=986
        keys [[1800,0],[1801,0.0613],[1802,0.2308],[1803,0.4848],[1804,0.797],[1805,1.1405],[1806,1.489],[1807,1.8183],[1808,2.1071],[1809,2.3385],[1810,2.5001],[1811,2.5849],[1812,2.591],[1813,2.5216],[1814,2.3844],[1815,2.1899],[1816,1.9506],[1817,1.6791],[1818,1.386],[1819,1.0775],[1820,0.7528],[1821,0.4013],[1822,0],[1823,-0.6119],[1824,-1.5587],[1825,-2.8178],[1826,-4.3653],[1827,-6.1755],[1828,-8.2212],[1829,-10.474],[1830,-12.905],[1831,-15.4852],[1832,-18.1865],[1833,-20.9815],[1838,-35.51],[1843,-50.5332],[1850,-72.2526],[1851,-75.51],[1852,-79.0205],[1853,-82.7593],[1854,-86.7018],[1856,-95.0981],[1866,-140.8564],[1868,-149.3745],[1870,-157.2195],[1872,-164.1938],[1874,-170.1007],[1876,-174.7436],[1877,-176.5299],[1878,-177.9268],[1895,-189.6474],[1904,-192.5029],[1982,-169.2186],[1990,-168.8383],[1993,-170.2233],[2003,-179.3006],[2006,-180.9153],[2039,-181.1469],[2040,-179.3946],[2041,-167.081],[2042,-146.8108],[2043,-123.1301],[2044,-100.5853],[2045,-83.7236],[2046,-77.0925],[2047,-77.2213],[2048,-77.5955],[2049,-78.1971],[2050,-79.0076],[2051,-80.0091],[2053,-82.5121],[2055,-85.5605],[2060,-94.614],[2061,-97.7006],[2062,-107.9168],[2063,-123.6311],[2064,-141.9148],[2065,-159.8392],[2066,-174.4756],[2067,-183.9111],[2068,-192.3107],[2069,-200.3158],[2070,-207.6631],[2071,-214.0892],[2072,-219.3307],[2087,-276.6416],[2088,-280.8213],[2090,-291.058],[2098,-341.432],[2100,-352.1265],[2101,-356.6038],[2118,-408.0407],[2123,-428.9698],[2129,-458.8967],[2137,-509.5527],[2139,-519.8512],[2141,-527.207],[2153,-540.2236],[2204,-541.4163],[2215,-421.0493],[2241,-417.0779],[2245,-420.1765],[2261,-445.416],[2267,-449.786],[2584,-447.8401],[2585,-449.0033],[2588,-456.4292],[2594,-489.4999],[2601,-523.424],[2614,-538.1965],[2820,-538.1965]]
       }
       body_offset_ctrl.Rotation.Y {
        type=MovieSceneScriptingFloatChannel
        default=0
        thinned_from=976
        keys [[1800,0],[1801,-0.0076],[1802,-0.031],[1803,-0.0728],[1804,-0.1364],[1805,-0.2253],[1806,-0.3409],[1807,-0.4827],[1808,-0.6473],[1809,-0.8289],[1811,-1.2078],[1812,-1.3839],[1813,-1.5355],[1814,-1.6509],[1815,-1.7186],[1816,-1.7277],[1817,-1.6676],[1818,-1.5281],[1819,-1.2993],[1820,-0.9724],[1821,-0.5401],[1822,0],[1823,0.5808],[1824,1.1356],[1825,1.662],[1826,2.1578],[1827,2.6218],[1828,3.054],[1829,3.4567],[1830,3.8341],[1831,4.1926],[1833,4.8883],[1834,5.2462],[1835,5.6256],[1836,6.0372],[1837,6.4906],[1838,7.021],[1839,7.6419],[1840,8.3303],[1841,9.0629],[1843,10.568],[1844,11.294],[1845,11.9714],[1846,12.5769],[1847,13.0874],[1848,13.4796],[1849,13.7304],[1850,13.8165],[1851,13.7513],[1852,13.5691],[1853,13.2786],[1854,12.8882],[1855,12.4062],[1856,11.8411],[1857,11.2014],[1858,10.4955],[1859,9.7318],[1860,8.9188],[1861,8.0651],[1862,7.179],[1863,6.2691],[1866,3.4828],[1867,2.564],[1868,1.6645],[1869,0.7932],[1870,-0.0412],[1871,-0.83],[1872,-1.5643],[1873,-2.2354],[1874,-2.8344],[1875,-3.3525],[1876,-3.781],[1877,-4.1109],[1878,-4.3335],[1879,-4.4401],[1880,-4.4771],[1881,-4.4978],[1882,-4.503],[1883,-4.4934],[1884,-4.4699],[1885,-4.4332],[1886,-4.3843],[1887,-4.3238],[1888,-4.2527],[1889,-4.1718],[1890,-4.0818],[1892,-3.8783],[1894,-3.6489],[1897,-3.2714],[1902,-2.6125],[1904,-2.3595],[1905,-2.2389],[1906,-2.1235],[1908,-1.9011],[1910,-1.67],[1912,-1.4304],[1914,-1.184],[1916,-0.9321],[1918,-0.6763],[1923,-0.0291],[1924,0.1002],[1925,0.2289],[1926,0.3569],[1927,0.484],[1928,0.6101],[1930,0.8582],[1932,1.0997],[1934,1.333],[1936,1.5566],[1938,1.7689],[1940,1.9684],[1942,2.1534],[1944,2.3225],[1957,3.3195],[1963,3.7382],[1968,4.0488],[1973,4.3132],[1977,4.4842],[1981,4.6131],[1984,4.6788],[1987,4.7155],[1989,4.7227],[1990,4.6602],[1991,4.4822],[1992,4.2026],[1993,3.8355],[1994,3.3949],[1995,2.8947],[1996,2.349],[1997,1.7717],[1998,1.177],[1999,0.5787],[2000,-0.009],[2001,-0.5722],[2002,-1.0969],[2003,-1.5691],[2004,-1.9748],[2005,-2.2999],[2006,-2.5304],[2007,-2.6524],[2008,-2.6683],[2039,-2.6683],[2040,-2.7203],[2041,-3.0858],[2042,-3.6883],[2043,-4.394],[2044,-5.0716],[2045,-5.5923],[2046,-5.8288],[2053,-6.1768],[2056,-6.2885],[2059,-6.3563],[2060,-6.3667],[2061,-6.3373],[2062,-6.0894],[2063,-5.6342],[2064,-5.0179],[2065,-4.2865],[2066,-3.4858],[2067,-2.662],[2068,-1.861],[2069,-1.1289],[2070,-0.5116],[2071,-0.0552],[2072,0.1944],[2073,0.2253],[2074,0.2148],[2075,0.1964],[2076,0.1713],[2077,0.1408],[2078,0.106],[2079,0.0682],[2080,0.0285],[2081,-0.0118],[2082,-0.0516],[2083,-0.0896],[2084,-0.1246],[2085,-0.1555],[2086,-0.181],[2087,-0.1999],[2091,-0.2619],[2097,-0.3591],[2155,-1.3797],[2174,-1.6602],[2189,-1.8299],[2202,-1.9281],[2213,-1.9685],[2218,-1.9717],[2219,-1.9662],[2220,-1.9556],[2221,-1.9405],[2223,-1.8988],[2225,-1.8449],[2232,-1.6199],[2234,-1.5618],[2236,-1.5139],[2237,-1.4951],[2238,-1.4802],[2239,-1.47],[2240,-1.4648],[2241,-1.5194],[2242,-1.9229],[2243,-2.6377],[2244,-3.5681],[2245,-4.6182],[2246,-5.6922],[2247,-6.6943],[2248,-7.5289],[2249,-8.1],[2250,-8.3118],[2251,-8.0384],[2252,-7.2583],[2253,-6.0323],[2254,-4.4199],[2255,-2.4808],[2256,-0.275],[2257,2.1367],[2258,4.692],[2259,7.3274],[2260,9.9775],[2261,12.5757],[2262,15.0541],[2263,17.3441],[2264,19.3765],[2265,21.0824],[2266,22.3929],[2267,23.2397],[2268,23.5549],[2269,23.3577],[2270,22.7529],[2271,21.8016],[2272,20.565],[2273,19.1041],[2274,17.48],[2275,15.7538],[2277,12.2397],[2278,10.574],[2279,9.0507],[2280,7.7308],[2281,6.6755],[2282,5.9459],[2283,5.6031],[2284,5.4809],[2285,5.3719],[2286,5.2754],[2287,5.1904],[2289,5.0514],[2291,4.9463],[2293,4.8667],[2296,4.7831],[2298,4.75],[2300,4.7403],[2344,5.4321],[2354,5.4647],[2365,5.4486],[2386,5.6366],[2391,5.6582],[2399,5.2979],[2421,4.8852],[2422,4.8603],[2425,4.7525],[2426,4.7258],[2427,4.7125],[2442,4.7668],[2444,4.7463],[2452,4.5751],[2455,4.5384],[2490,4.6827],[2494,4.7379],[2497,4.8083],[2500,4.9096],[2503,5.044],[2507,5.272],[2524,6.4717],[2530,6.8118],[2535,7.0179],[2542,7.192],[2583,7.2826],[2584,7.2471],[2585,7.1442],[2588,0.0222],[2590,-4.1028],[2594,-7.3746],[2601,-2.8844],[2614,-1.8216],[2820,-1.8216]]
       }
       body_offset_ctrl.Rotation.Z {
        type=MovieSceneScriptingFloatChannel
        default=0
        thinned_from=975
        keys [[1800,0],[1801,0.1897],[1802,0.7233],[1803,1.5475],[1804,2.6097],[1805,3.8569],[1806,5.2358],[1807,6.6923],[1808,8.1718],[1809,9.6193],[1810,10.9789],[1811,12.1946],[1812,13.2098],[1813,13.968],[1814,14.4127],[1815,14.4875],[1816,14.1364],[1817,13.3037],[1818,11.9345],[1819,9.9739],[1820,7.3677],[1821,4.0615],[1822,0],[1823,-4.9055],[1824,-10.5977],[1825,-16.9137],[1826,-23.6906],[1827,-30.7651],[1829,-45.1534],[1830,-52.1398],[1831,-58.7695],[1832,-64.8783],[1833,-70.3024],[1834,-74.8781],[1835,-78.4414],[1836,-80.8284],[1837,-81.8752],[1850,-86.0911],[1851,-86.511],[1852,-87.1285],[1853,-87.933],[1855,-90.0603],[1857,-92.8074],[1859,-96.0888],[1862,-101.8262],[1876,-133.652],[1891,-162.0682],[1899,-174.4706],[1905,-181.4302],[1963,-220.433],[1987,-227.0722],[2040,-226.7498],[2041,-225.1844],[2044,-216.7234],[2045,-214.5758],[2046,-213.7308],[2084,-228.33],[2250,-229.2496],[2251,-230.0611],[2252,-232.0734],[2253,-235.1411],[2254,-239.1182],[2255,-243.8585],[2257,-255.0461],[2262,-286.2937],[2264,-297.1612],[2265,-301.6913],[2266,-305.4339],[2267,-308.2492],[2284,-329.4618],[2286,-333.8852],[2289,-343.3536],[2297,-373.2609],[2299,-379.0736],[2301,-383.2146],[2312,-395.4822],[2334,-407.0838],[2362,-453.1732],[2387,-468.0515],[2392,-472.1363],[2399,-491.7372],[2422,-528.4847],[2426,-537.8411],[2443,-548.2076],[2457,-579.9684],[2491,-612.436],[2538,-741.224],[2551,-758.6367],[2584,-759.3542],[2587,-753.4439],[2589,-746.3541],[2593,-727.129],[2598,-699.3717],[2601,-685.1437],[2614,-664.1074],[2820,-664.1074]]
       }
       body_offset_ctrl.Scale.X {type=MovieSceneScriptingFloatChannel, default=1, const [1,1006]}
       body_offset_ctrl.Scale.Y {type=MovieSceneScriptingFloatChannel, default=1, const [1,1006]}
       body_offset_ctrl.Scale.Z {type=MovieSceneScriptingFloatChannel, default=1, const [1,1006]}
       thigh_l_fk_ctrl.Location.X {type=MovieSceneScriptingFloatChannel, default=0, const [0,15]}
       thigh_l_fk_ctrl.Location.Y {type=MovieSceneScriptingFloatChannel, default=0, const [0,15]}
       thigh_l_fk_ctrl.Location.Z {type=MovieSceneScriptingFloatChannel, default=0, const [0,15]}
       thigh_l_fk_ctrl.Rotation.X {
        type=MovieSceneScriptingFloatChannel
        default=16.0256
        thinned_from=19
        keys [[2162,16.0256],[2180,5.339],[2202,15.8993],[2208,5.8759],[2223,0.0673],[2586,0.1106],[2598,-3.1257],[2606,8.4879],[2617,16.0256],[2644,16.0256],[2660,6.4483],[2677,16.0256],[2734,16.0256],[2752,15.3178],[2754,15.316],[2759,-0.0628]]
       }
       thigh_l_fk_ctrl.Rotation.Y {
        type=MovieSceneScriptingFloatChannel
        default=-2.1893
        thinned_from=19
        keys [[2162,-2.1893],[2180,-5.7277],[2202,6.1971],[2208,5.4327],[2223,-7.1055],[2586,-7.0942],[2598,-4.6379],[2606,-2.8759],[2617,-2.1893],[2644,-2.1893],[2660,-1.1896],[2677,-2.1893],[2734,-2.1893],[2748,11.9043],[2752,12.8482],[2754,12.8667],[2759,-7.1543]]
       }
       thigh_l_fk_ctrl.Rotation.Z {
        type=MovieSceneScriptingFloatChannel
        default=-14.05
        thinned_from=19
        keys [[2162,-14.05],[2180,-18.6933],[2202,-51.2877],[2208,-70.4512],[2223,2.0803],[2586,1.9094],[2598,14.009],[2606,-3.0063],[2617,-14.05],[2644,-14.05],[2660,-23.7748],[2677,-14.05],[2734,-14.05],[2748,-64.4216],[2752,-72.2315],[2754,-73.0667],[2759,3.2829]]
       }
       thigh_l_fk_ctrl.Scale.X {type=MovieSceneScriptingFloatChannel, default=1, const [1,2]}
       thigh_l_fk_ctrl.Scale.Y {type=MovieSceneScriptingFloatChannel, default=1, const [1,2]}
       thigh_l_fk_ctrl.Scale.Z {type=MovieSceneScriptingFloatChannel, default=1, const [1,2]}
       calf_l_fk_ctrl.Location.X {type=MovieSceneScriptingFloatChannel, default=0, const [0,15]}
       calf_l_fk_ctrl.Location.Y {type=MovieSceneScriptingFloatChannel, default=0, const [0,15]}
       calf_l_fk_ctrl.Location.Z {type=MovieSceneScriptingFloatChannel, default=0, const [0,15]}
       calf_l_fk_ctrl.Rotation.X {
        type=MovieSceneScriptingFloatChannel
        default=1.0313
        thinned_from=19
        keys [[2162,1.0313],[2180,0.9309],[2202,1.1747],[2208,0.3431],[2223,0],[2586,0.0077],[2598,0.2175],[2606,0.7932],[2617,1.0313],[2644,1.0313],[2660,1.1262],[2677,1.0313],[2734,1.0313],[2748,1.1368],[2752,1.1131],[2754,1.1131],[2759,0.0078]]
       }
       calf_l_fk_ctrl.Rotation.Y {
        type=MovieSceneScriptingFloatChannel
        default=0.5628
        thinned_from=19
        keys [[2162,0.5628],[2180,0.7167],[2202,-0.0192],[2208,-0.4768],[2223,0],[2586,0.006],[2598,-0.0095],[2606,0.2157],[2617,0.5628],[2644,0.5628],[2660,0.3344],[2677,0.5628],[2734,0.5628],[2748,-0.2854],[2752,-0.3745],[2754,-0.3745],[2759,0.0059]]
       }
       calf_l_fk_ctrl.Rotation.Z {
        type=MovieSceneScriptingFloatChannel
        default=83.4505
        thinned_from=19
        keys [[2162,83.4505],[2180,74.484],[2202,113.0081],[2208,109.819],[2223,0],[2586,0.6129],[2598,46.8845],[2606,82.7727],[2617,83.4505],[2644,83.4505],[2660,95.535],[2677,83.4505],[2734,83.4505],[2748,128.0941],[2752,130.6609],[2754,130.6609],[2759,1.9763]]
       }
       calf_l_fk_ctrl.Scale.X {type=MovieSceneScriptingFloatChannel, default=1, const [1,2]}
       calf_l_fk_ctrl.Scale.Y {type=MovieSceneScriptingFloatChannel, default=1, const [1,2]}
       calf_l_fk_ctrl.Scale.Z {type=MovieSceneScriptingFloatChannel, default=1, const [1,2]}
       foot_l_fk_ctrl.Location.X {type=MovieSceneScriptingFloatChannel, default=0, const [0,15]}
       foot_l_fk_ctrl.Location.Y {type=MovieSceneScriptingFloatChannel, default=0, const [0,15]}
       foot_l_fk_ctrl.Location.Z {type=MovieSceneScriptingFloatChannel, default=0, const [0,15]}
       foot_l_fk_ctrl.Rotation.X {
        type=MovieSceneScriptingFloatChannel
        default=-1.478
        thinned_from=18
        keys [[2162,-1.478],[2202,-1.478],[2208,-2.8568],[2223,-5.395],[2586,-5.3627],[2598,-4.6691],[2606,-2.6066],[2617,-1.478],[2754,-1.478],[2759,-5.3627]]
       }
       foot_l_fk_ctrl.Rotation.Y {
        type=MovieSceneScriptingFloatChannel
        default=-2.2009
        thinned_from=18
        keys [[2162,-2.2009],[2202,-2.2009],[2208,-1.0846],[2223,0.9704],[2586,0.9443],[2598,0.3827],[2606,-1.2872],[2617,-2.2009],[2754,-2.2009],[2759,0.9443]]
       }
       foot_l_fk_ctrl.Rotation.Z {
        type=MovieSceneScriptingFloatChannel
        default=21.0968
        thinned_from=18
        keys [[2162,21.0968],[2202,21.0968],[2208,26.5541],[2223,36.6006],[2586,36.473],[2598,33.7276],[2606,25.5638],[2617,21.0968],[2754,21.0968],[2759,36.473]]
       }
       foot_l_fk_ctrl.Scale.X {type=MovieSceneScriptingFloatChannel, default=1, const [1,2]}
       foot_l_fk_ctrl.Scale.Y {type=MovieSceneScriptingFloatChannel, default=1, const [1,2]}
       foot_l_fk_ctrl.Scale.Z {type=MovieSceneScriptingFloatChannel, default=1, const [1,2]}
       ball_l_fk_ctrl.Location.X {type=MovieSceneScriptingFloatChannel, default=0, const [0,1021]}
       ball_l_fk_ctrl.Location.Y {type=MovieSceneScriptingFloatChannel, default=0, const [0,1021]}
       ball_l_fk_ctrl.Location.Z {type=MovieSceneScriptingFloatChannel, default=0, const [0,1021]}
       ball_l_fk_ctrl.Rotation.X {type=MovieSceneScriptingFloatChannel, default=0, const [0,1021]}
       ball_l_fk_ctrl.Rotation.Y {type=MovieSceneScriptingFloatChannel, default=0, const [0,1021]}
       ball_l_fk_ctrl.Rotation.Z {type=MovieSceneScriptingFloatChannel, default=0, const [0,1021]}
       ball_l_fk_ctrl.Scale.X {type=MovieSceneScriptingFloatChannel, default=1, const [1,1021]}
       ball_l_fk_ctrl.Scale.Y {type=MovieSceneScriptingFloatChannel, default=1, const [1,1021]}
       ball_l_fk_ctrl.Scale.Z {type=MovieSceneScriptingFloatChannel, default=1, const [1,1021]}
       foot_l_ik_ctrl.Location.X {
        type=MovieSceneScriptingFloatChannel
        default=0
        thinned_from=1021
        keys [[1800,0],[1801,-16.9412],[1802,-34.5142],[1803,-52.5518],[1804,-70.905],[1808,-145.3074],[1809,-163.6722],[1810,-181.6926],[1811,-199.1921],[1812,-215.9741],[1813,-231.845],[1814,-246.6467],[1815,-260.2955],[1816,-272.8202],[1817,-284.3953],[1819,-306.2248],[1820,-317.6458],[1821,-330.3895],[1822,-345.2629],[1826,-411.8088],[1834,-551.4203],[1837,-600.12],[1849,-683.7155],[1852,-700.9842],[1862,-741.6907],[1865,-758.9789],[1867,-774.4854],[1869,-794.2222],[1871,-819.08],[1873,-849.9301],[1874,-867.8672],[1876,-908.8591],[1879,-978.5137],[1895,-1407.8916],[1899,-1504.0742],[1902,-1566.439],[1905,-1617.8901],[1934,-1962.4141],[1984,-2304.9512],[2040,-2436.8403],[2041,-2431.8928],[2043,-2406.4995],[2045,-2366.9658],[2047,-2313.0491],[2049,-2245.0381],[2051,-2164.791],[2053,-2074.031],[2056,-1921.9509],[2060,-1698.4673],[2063,-1522.9556],[2064,-1469.4645],[2065,-1422.5916],[2066,-1384.4976],[2069,-1296.4158],[2070,-1269.9576],[2071,-1249.3038],[2072,-1237.3013],[2073,-1236.7876],[2074,-1248.9229],[2075,-1272.5109],[2076,-1306.2008],[2077,-1348.6421],[2078,-1398.4834],[2079,-1454.3745],[2080,-1514.9648],[2085,-1841.1565],[2086,-1901.5962],[2090,-2113.8005],[2093,-2258.0649],[2095,-2341.6204],[2097,-2411.1179],[2099,-2463.1833],[2100,-2481.6147],[2101,-2494.4028],[2102,-2501.1113],[2107,-2490.272],[2116,-2418.3218],[2117,-2404.1482],[2118,-2381.4092],[2119,-2351.2178],[2120,-2314.6877],[2122,-2227.062],[2127,-1974.8405],[2131,-1792.5386],[2133,-1708.1523],[2135,-1632.1051],[2136,-1598.209],[2137,-1567.5945],[2138,-1540.6611],[2139,-1517.8109],[2140,-1499.4501],[2141,-1485.9884],[2142,-1477.8383],[2157,-1451.4836],[2204,-1449.6624],[2205,-1447.4227],[2206,-1441.0565],[2208,-1418.9718],[2210,-1390.1761],[2212,-1353.5344],[2214,-1307.5264],[2216,-1252.6476],[2218,-1189.384],[2220,-1118.2136],[2222,-1039.6223],[2224,-954.0991],[2226,-862.1326],[2228,-764.2119],[2230,-660.8252],[2232,-552.4624],[2233,-496.5679],[2234,-439.6123],[2235,-381.6567],[2236,-322.7627],[2237,-262.9917],[2238,-202.4038],[2239,-141.061],[2240,-79.0239],[2241,-16.3535],[2242,46.8916],[2243,110.6533],[2244,174.873],[2247,369.6787],[2250,566.3457],[2251,636.9346],[2252,716.6494],[2253,804.3975],[2254,899.0889],[2255,999.6318],[2256,1104.9268],[2257,1213.8672],[2262,1774.3369],[2263,1881.873],[2264,1985.3838],[2265,2083.8281],[2266,2176.1816],[2267,2261.4238],[2278,3092.2373],[2281,3299.916],[2283,3425.2236],[2285,3531.2324],[2287,3611.7207],[2289,3667.2061],[2291,3698.2471],[2293,3705.4756],[2295,3689.6133],[2297,3651.4561],[2299,3591.8501],[2301,3511.6621],[2312,2972.2322],[2322,2441.7576],[2326,2250.6248],[2329,2123.0344],[2332,2009.1929],[2333,1966.2034],[2334,1918.6272],[2335,1866.6729],[2336,1810.5479],[2337,1750.4612],[2338,1686.621],[2339,1619.2355],[2340,1548.5137],[2342,1397.8948],[2344,1236.436],[2346,1065.8098],[2348,887.6909],[2350,703.7534],[2352,515.6704],[2357,38.2915],[2358,-56.7471],[2359,-151.1543],[2360,-244.7236],[2361,-337.2466],[2362,-428.5176],[2363,-518.3286],[2364,-606.4727],[2365,-692.7432],[2366,-776.9336],[2375,-1498.666],[2380,-1883.3975],[2385,-2246.4766],[2389,-2515.7559],[2393,-2762.2842],[2396,-2930.1152],[2399,-3079.7764],[2402,-3209.0625],[2406,-3352.2158],[2412,-3518.7598],[2435,-3982.2117],[2441,-4054.9382],[2445,-4068.8105],[2449,-4045.6169],[2452,-3999.6472],[2455,-3926.042],[2458,-3822.0723],[2462,-3637.24],[2468,-3306.4895],[2480,-2590.3955],[2484,-2373.7529],[2488,-2181.5039],[2489,-2128.2695],[2490,-2070.0986],[2491,-2007.4492],[2493,-1870.5596],[2495,-1721.2842],[2498,-1482.2207],[2502,-1154.583],[2504,-995.2744],[2505,-918.3652],[2506,-843.9014],[2507,-772.3447],[2508,-704.1543],[2509,-639.792],[2510,-579.7168],[2511,-524.3877],[2512,-474.2627],[2513,-429.1582],[2514,-385.5234],[2515,-342.7656],[2516,-300.917],[2517,-260.0068],[2518,-220.0664],[2519,-181.125],[2520,-143.2119],[2521,-106.3564],[2522,-70.5889],[2523,-35.9346],[2524,-2.4248],[2525,29.915],[2526,61.0566],[2527,90.9717],[2528,119.6338],[2529,147.0166],[2530,173.0918],[2531,197.833],[2532,221.2139],[2533,243.208],[2534,263.7871],[2535,282.9248],[2536,300.5947],[2537,316.7695],[2538,331.4229],[2539,344.5264],[2540,356.0527],[2541,365.9756],[2542,374.2656],[2543,380.8965],[2544,385.8398],[2545,389.0664],[2546,390.5488],[2547,390.2588],[2548,388.168],[2549,384.2471],[2550,378.4668],[2551,370.7998],[2552,361.2148],[2553,349.6846],[2554,336.1787],[2555,320.668],[2556,303.1572],[2557,283.7285],[2558,262.4678],[2559,239.4609],[2560,214.7954],[2561,188.5581],[2562,160.8345],[2563,131.7114],[2564,101.2754],[2565,69.6133],[2566,36.811],[2567,2.9551],[2568,-31.8677],[2569,-67.5708],[2570,-104.0679],[2571,-141.2725],[2572,-179.0977],[2573,-217.4575],[2575,-295.4351],[2583,-611.4253],[2585,-687.3845],[2587,-759.9597],[2589,-828.804],[2591,-893.7147],[2593,-954.5997],[2595,-1011.3359],[2597,-1063.5708],[2599,-1110.5811],[2601,-1151.2836],[2603,-1184.3854],[2605,-1208.5754],[2606,-1216.9478],[2607,-1222.6534],[2608,-1225.5646],[2757,-1225.7076],[2759,-1220.4114],[2761,-1208.4797],[2763,-1190.4335],[2765,-1166.7926],[2767,-1138.0778],[2770,-1086.6301],[2773,-1026.6934],[2777,-936.6078],[2790,-611.4831],[2793,-539.837],[2795,-494.7292],[2797,-452.3535],[2798,-432.353],[2799,-422.6779],[2820,-422.6779]]
       }
       foot_l_ik_ctrl.Location.Y {
        type=MovieSceneScriptingFloatChannel
        default=0
        thinned_from=1021
        keys [[1800,0],[1801,11.1332],[1802,22.8989],[1803,35.9454],[1804,50.7526],[1805,67.6112],[1806,86.615],[1807,107.6688],[1808,130.5105],[1809,154.745],[1810,179.8882],[1812,230.8204],[1813,255.6465],[1814,279.5413],[1815,302.2764],[1816,323.7631],[1817,344.0519],[1818,363.3196],[1830,583.8],[1837,719.4905],[1848,610.0591],[1850,588.5154],[1851,574.9734],[1852,556.4375],[1853,533.2976],[1854,505.9453],[1855,474.7695],[1856,440.1589],[1857,402.5],[1858,362.178],[1859,319.5764],[1860,275.0781],[1861,229.0637],[1862,181.9146],[1863,134.0103],[1865,37.4609],[1866,-10.4202],[1867,-57.5276],[1868,-103.4751],[1869,-147.8756],[1870,-190.3396],[1871,-230.476],[1872,-267.8925],[1873,-302.1959],[1874,-332.9924],[1875,-359.8875],[1876,-384.2603],[1877,-407.7499],[1878,-430.3522],[1880,-472.896],[1882,-511.9752],[1884,-547.6169],[1886,-579.8354],[1888,-608.6465],[1890,-634.0671],[1892,-656.1165],[1894,-674.8145],[1896,-690.1831],[1898,-702.2446],[1900,-711.0229],[1902,-716.5415],[1904,-718.8245],[1906,-717.8955],[1943,-645.1794],[1993,-503.4214],[2003,-464.7871],[2014,-410.6104],[2035,-288.394],[2039,-263.8574],[2040,-256.5884],[2041,-241.6387],[2042,-218.3701],[2043,-187.7705],[2044,-151.2793],[2045,-109.9204],[2046,-63.8755],[2047,-13.0498],[2048,42.3193],[2049,101.7705],[2050,164.8433],[2051,231.0762],[2052,300.0073],[2053,371.1763],[2054,444.1206],[2059,819.3594],[2061,965.5024],[2062,1033.1934],[2063,1093.8691],[2064,1145.4731],[2065,1186.2515],[2066,1214.4775],[2067,1229.77],[2068,1240.3696],[2069,1248.0547],[2071,1255.8491],[2072,1256.5093],[2073,1254.1572],[2074,1242.1479],[2075,1220.375],[2076,1189.9746],[2077,1152.0811],[2078,1107.8311],[2079,1058.3589],[2080,1004.8008],[2082,889.9629],[2084,772.3965],[2085,715.4272],[2086,661.1812],[2087,610.791],[2089,515.5127],[2090,469.1377],[2091,424.1353],[2092,380.8945],[2093,339.8008],[2094,301.2329],[2095,265.5684],[2096,233.1787],[2097,204.4365],[2098,179.7163],[2099,159.394],[2100,143.8496],[2101,133.4658],[2102,128.6289],[2103,129.0874],[2104,131.1929],[2105,134.2358],[2106,138.1304],[2107,142.792],[2108,148.1338],[2109,154.0703],[2110,160.5151],[2111,167.3843],[2113,182.0488],[2116,205.082],[2117,216.9531],[2118,236.5596],[2119,262.9072],[2120,295.001],[2121,331.8491],[2122,372.4551],[2123,415.8286],[2126,552.646],[2127,597.1904],[2128,639.5605],[2130,720.5386],[2132,797.7827],[2133,834.1499],[2134,868.5688],[2135,900.7129],[2136,930.2573],[2137,956.8774],[2138,980.2471],[2139,1000.0361],[2140,1015.9126],[2141,1027.541],[2142,1034.5801],[2154,1055.3491],[2204,1055.7607],[2205,1058.0488],[2206,1064.583],[2207,1074.7896],[2209,1102.1201],[2211,1136.2344],[2212,1156.8008],[2213,1179.8237],[2215,1232.9751],[2217,1295.167],[2219,1365.876],[2221,1444.5806],[2223,1530.7549],[2225,1623.876],[2228,1775.4365],[2232,1996.9385],[2238,2362.2969],[2252,3290.1694],[2254,3406.0625],[2256,3503.4971],[2258,3578.1807],[2260,3626.0317],[2261,3638.7039],[2262,3643.2769],[2263,3639.3213],[2264,3626.4165],[2265,3604.1443],[2266,3572.083],[2267,3529.8074],[2268,3476.8828],[2270,3351.8787],[2273,3139.5254],[2281,2518.7188],[2284,2297.1104],[2285,2214.6157],[2286,2125.2908],[2287,2029.8838],[2288,1929.1431],[2289,1823.8213],[2291,1602.437],[2297,906.3389],[2298,793.4619],[2299,683.4385],[2300,576.9932],[2301,474.8506],[2302,377.0566],[2304,183.6943],[2305,88.1318],[2306,-6.2402],[2307,-99.0879],[2308,-190.0791],[2309,-278.8789],[2310,-365.1553],[2311,-448.5762],[2312,-528.8066],[2313,-605.5156],[2314,-678.3691],[2318,-959.835],[2321,-1164.0576],[2323,-1294.1631],[2325,-1417.7119],[2327,-1533.2285],[2329,-1639.2373],[2331,-1734.2607],[2332,-1776.832],[2333,-1815.6113],[2335,-1882.4229],[2337,-1935.9531],[2339,-1977.459],[2341,-2008.1973],[2343,-2029.4287],[2346,-2046.2139],[2350,-2047.1577],[2362,-2000.8391],[2365,-2000.9026],[2376,-2046.0457],[2381,-2044.7021],[2384,-2030.5601],[2387,-2003.67],[2390,-1961.785],[2392,-1924.0612],[2394,-1877.0068],[2396,-1820.4448],[2398,-1754.4253],[2400,-1678.5703],[2402,-1590.1748],[2404,-1489.8984],[2406,-1378.9932],[2408,-1258.7114],[2410,-1130.3057],[2412,-995.0273],[2414,-854.1299],[2416,-708.8652],[2419,-485.5176],[2421,-334.8125],[2422,-258.1768],[2423,-179.3291],[2424,-98.5],[2425,-15.9131],[2426,68.2129],[2427,153.665],[2428,240.2734],[2429,328.083],[2430,416.9893],[2432,597.5352],[2443,1609.2988],[2447,1962.4492],[2450,2213.1582],[2453,2447.1484],[2456,2660.4551],[2458,2789.2969],[2476,3853.5566],[2481,4109.3262],[2485,4279],[2488,4379.0879],[2491,4442.7285],[2494,4470.0156],[2497,4465.0479],[2501,4415.3721],[2506,4297.9922],[2515,3990.3223],[2540,2938.9165],[2546,2721.2271],[2550,2597.8379],[2554,2496.3667],[2558,2404.6855],[2562,2288.292],[2567,2113.9048],[2585,1380.7856],[2588,1267.5229],[2590,1197.6182],[2592,1133.1943],[2594,1074.835],[2596,1022.8179],[2598,977.2183],[2600,938.1235],[2602,905.8423],[2603,892.4365],[2604,880.9849],[2605,871.5991],[2606,864.397],[2607,859.4937],[2608,856.9966],[2757,856.3682],[2758,853.9014],[2759,849.0542],[2760,841.916],[2761,832.5776],[2762,821.1274],[2763,807.6567],[2764,792.2539],[2765,775.0103],[2766,756.0142],[2767,735.3564],[2768,713.127],[2769,689.415],[2771,637.9038],[2773,581.542],[2775,521.0474],[2777,457.1392],[2779,390.5361],[2781,321.957],[2787,111.5479],[2788,76.7417],[2789,42.2495],[2790,8.1621],[2791,-25.4312],[2792,-58.4409],[2793,-90.7769],[2794,-122.3486],[2795,-153.0679],[2796,-182.8433],[2797,-211.5859],[2798,-239.2056],[2799,-252.5659],[2820,-252.5659]]
       }
       foot_l_ik_ctrl.Location.Z {
        type=MovieSceneScriptingFloatChannel
        default=0
        thinned_from=1021
        keys [[1800,0],[1801,114.8301],[1802,232.8555],[1803,353.8223],[1804,477.4805],[1805,603.5918],[1806,731.9238],[1807,862.2695],[1809,1128.2754],[1812,1538.502],[1823,3088.8438],[1827,3626.0684],[1830,4006.1357],[1833,4360.2402],[1836,4682.7344],[1839,4969.3442],[1842,5213.6807],[1844,5349.8193],[1846,5462.5269],[1848,5550.062],[1892,6518.3271],[1905,6636.5171],[2041,6417.1929],[2057,6707.0664],[2060,6727.5737],[2061,6724.6021],[2062,6704.292],[2064,6621.814],[2067,6445.3193],[2068,6354.4595],[2071,6021.7583],[2072,5933.9092],[2078,5612.7095],[2081,5487.1123],[2084,5402.2534],[2086,5375.3369],[2089,5383.9468],[2092,5434.3247],[2098,5617.8076],[2115,6330.5371],[2123,6550.8652],[2127,6600.1465],[2130,6585.9858],[2134,6491.2192],[2149,6042.6704],[2190,5557.8081],[2205,5513.1084],[2208,5538.0146],[2214,5661.6519],[2222,5929.8667],[2248,7080.583],[2261,7492.96],[2266,7577.8242],[2270,7584.0107],[2282,7399.3857],[2756,7674.6895],[2757,7658.437],[2758,7530.168],[2759,7278.105],[2760,6906.9214],[2761,6421.2876],[2762,5825.875],[2763,5125.356],[2764,4324.4004],[2765,3427.6816],[2766,2439.8691],[2767,1365.6367],[2768,209.6543],[2769,-1023.4062],[2770,-2328.8789],[2771,-3702.082],[2772,-5138.3516],[2773,-6633.0117],[2774,-8181.3984],[2775,-9778.8281],[2776,-11420.6406],[2777,-13102.1562],[2778,-14818.7188],[2780,-18338.2422],[2789,-34677.1406],[2792,-39913.2344],[2795,-44834],[2797,-47877.0469],[2798,-49313.2969],[2799,-50008.0625],[2820,-50008.0625]]
       }
       foot_l_ik_ctrl.Rotation.X {
        type=MovieSceneScriptingFloatChannel
        default=0
        thinned_from=1021
        keys [[1800,0],[1801,0.0613],[1802,0.2308],[1803,0.4848],[1804,0.797],[1805,1.1405],[1806,1.489],[1807,1.8183],[1808,2.1071],[1809,2.3385],[1810,2.5001],[1811,2.5849],[1812,-357.409],[1819,-358.9225],[1820,0.7528],[1821,0.4013],[1822,0],[1823,-0.6119],[1824,-1.5587],[1825,-2.8178],[1826,-4.3653],[1827,-6.1755],[1828,-8.2212],[1829,-10.474],[1830,-12.905],[1831,344.5148],[1851,284.49],[1854,273.2982],[1867,214.8127],[1870,202.7805],[1872,195.8062],[1874,189.8993],[1876,185.2564],[1877,183.4701],[1878,182.0732],[1894,170.8593],[1903,167.64],[1925,170.6744],[1980,190.4786],[1990,191.1617],[1993,189.7767],[2003,180.6994],[2006,179.0847],[2039,178.8531],[2040,180.6054],[2041,192.919],[2042,213.1892],[2043,236.8699],[2044,259.4147],[2045,276.2764],[2046,282.9075],[2049,281.8029],[2053,277.4879],[2060,265.386],[2061,262.2994],[2062,252.0832],[2063,236.3689],[2065,200.1608],[2066,185.5244],[2067,176.0889],[2068,167.6893],[2069,159.6842],[2070,152.3369],[2071,145.9108],[2072,140.6693],[2087,83.3584],[2088,79.1787],[2089,74.3378],[2090,68.942],[2091,63.0977],[2092,56.9114],[2093,50.4894],[2095,37.3639],[2096,30.8733],[2097,24.5725],[2098,18.568],[2099,12.9662],[2100,7.8735],[2101,3.3962],[2102,359.6409],[2117,315.3839],[2120,304.261],[2124,286.288],[2128,266.4409],[2130,255.2536],[2136,216.3881],[2137,210.4473],[2138,204.9926],[2139,200.1488],[2140,196.0406],[2141,192.793],[2142,190.5305],[2149,182.7872],[2153,179.7764],[2156,178.6929],[2204,178.5837],[2205,185.1157],[2206,201.9741],[2207,225.0518],[2208,250.2417],[2209,273.4365],[2210,290.5291],[2211,297.4123],[2241,302.9221],[2244,300.9881],[2252,288.3457],[2253,-73.3931],[2256,-78.4047],[2259,-82.8784],[2261,-85.416],[2263,-87.4871],[2265,-88.9839],[2266,-89.4797],[2267,-89.786],[2285,-89.4173],[2299,-87.4989],[2302,-87.4983],[2305,-88.1855],[2311,-90.4838],[2313,-90.9294],[2339,-91.1505],[2340,268.8313],[2471,272.6685],[2472,-87.2987],[2583,-87.4381],[2584,-87.8401],[2585,-89.0033],[2586,-90.8631],[2587,-93.3554],[2588,-96.4162],[2589,-99.9819],[2590,-103.989],[2591,-108.3742],[2592,-113.0743],[2594,-123.1659],[2599,-149.4366],[2601,-158.9838],[2602,-163.2822],[2603,-167.1726],[2604,-170.5859],[2605,-173.4533],[2606,-175.7057],[2607,-177.2744],[2608,-178.0912],[2820,-178.1965]]
       }
       foot_l_ik_ctrl.Rotation.Y {
        type=MovieSceneScriptingFloatChannel
        default=0
        thinned_from=1021
        keys [[1800,0],[1801,-0.0076],[1802,-0.031],[1803,-0.0728],[1804,-0.1364],[1805,-0.2253],[1806,-0.3409],[1807,-0.4827],[1808,-0.6473],[1809,-0.8289],[1811,-1.2078],[1812,-1.3839],[1813,-1.5355],[1814,-1.6509],[1815,-1.7186],[1816,-1.7277],[1817,-1.6676],[1818,-1.5281],[1819,-1.2993],[1820,-0.9724],[1821,-0.5401],[1822,0],[1823,0.5808],[1824,1.1356],[1825,1.662],[1826,2.1578],[1827,2.6218],[1828,3.054],[1829,3.4567],[1830,3.8341],[1831,4.1926],[1833,4.8883],[1834,5.2462],[1835,5.6256],[1836,6.0372],[1837,6.4906],[1838,7.021],[1839,7.6419],[1840,8.3303],[1841,9.0629],[1843,10.568],[1844,11.294],[1845,11.9714],[1846,12.5769],[1847,13.0874],[1848,13.4796],[1849,13.7304],[1850,13.8165],[1851,13.7513],[1852,13.5691],[1853,13.2786],[1854,12.8882],[1855,12.4062],[1856,11.8411],[1857,11.2014],[1858,10.4955],[1859,9.7318],[1860,8.9188],[1861,8.0651],[1862,7.179],[1863,6.2691],[1866,3.4828],[1867,2.564],[1868,1.6645],[1869,0.7932],[1870,-0.0412],[1871,-0.83],[1872,-1.5643],[1873,-2.2354],[1874,-2.8344],[1875,-3.3525],[1876,-3.781],[1877,-4.1109],[1878,-4.3335],[1879,-4.4401],[1880,-4.4771],[1881,-4.4978],[1882,-4.503],[1883,-4.4934],[1884,-4.4699],[1885,-4.4332],[1886,-4.3843],[1887,-4.3238],[1888,-4.2527],[1889,-4.1718],[1890,-4.0818],[1892,-3.8783],[1894,-3.6489],[1897,-3.2714],[1902,-2.6125],[1904,-2.3595],[1905,-2.2389],[1906,-2.1235],[1908,-1.9011],[1910,-1.67],[1912,-1.4304],[1914,-1.184],[1916,-0.9321],[1918,-0.6763],[1923,-0.0291],[1924,0.1002],[1925,0.2289],[1926,0.3569],[1927,0.484],[1928,0.6101],[1930,0.8582],[1932,1.0997],[1934,1.333],[1936,1.5566],[1938,1.7689],[1940,1.9684],[1942,2.1534],[1944,2.3225],[1957,3.3195],[1963,3.7382],[1968,4.0488],[1973,4.3132],[1977,4.4842],[1981,4.6131],[1984,4.6788],[1987,4.7155],[1989,4.7227],[1990,4.6602],[1991,4.4822],[1992,4.2026],[1993,3.8355],[1994,3.3949],[1995,2.8947],[1996,2.349],[1997,1.7717],[1998,1.177],[1999,0.5787],[2000,-0.009],[2001,-0.5722],[2002,-1.0969],[2003,-1.5691],[2004,-1.9748],[2005,-2.2999],[2006,-2.5304],[2007,-2.6524],[2008,-2.6683],[2039,-2.6683],[2040,-2.7203],[2041,-3.0858],[2042,-3.6883],[2043,-4.394],[2044,-5.0716],[2045,-5.5923],[2046,-5.8288],[2053,-6.1768],[2056,-6.2885],[2059,-6.3563],[2060,-6.3667],[2061,-6.3373],[2062,-6.0894],[2063,-5.6342],[2064,-5.0179],[2065,-4.2865],[2066,-3.4858],[2067,-2.662],[2068,-1.861],[2069,-1.1289],[2070,-0.5116],[2071,-0.0552],[2072,0.1944],[2073,0.2253],[2074,0.2148],[2075,0.1964],[2076,0.1713],[2077,0.1408],[2078,0.106],[2079,0.0682],[2080,0.0285],[2081,-0.0118],[2082,-0.0516],[2083,-0.0896],[2084,-0.1246],[2085,-0.1555],[2086,-0.181],[2087,-0.1999],[2091,-0.2619],[2097,-0.3591],[2155,-1.3797],[2174,-1.6602],[2189,-1.8299],[2202,-1.9281],[2213,-1.9685],[2218,-1.9717],[2219,-1.9662],[2220,-1.9556],[2221,-1.9405],[2223,-1.8988],[2225,-1.8449],[2232,-1.6199],[2234,-1.5618],[2236,-1.5139],[2237,-1.4951],[2238,-1.4802],[2239,-1.47],[2240,-1.4648],[2241,-1.5194],[2242,-1.9229],[2243,-2.6377],[2244,-3.5681],[2245,-4.6182],[2246,-5.6922],[2247,-6.6943],[2248,-7.5289],[2249,-8.1],[2250,-8.3118],[2251,-8.0384],[2252,-7.2583],[2253,-6.0323],[2254,-4.4199],[2255,-2.4808],[2256,-0.275],[2257,2.1367],[2258,4.692],[2259,7.3274],[2260,9.9775],[2261,12.5757],[2262,15.0541],[2263,17.3441],[2264,19.3765],[2265,21.0824],[2266,22.3929],[2267,23.2397],[2268,23.5549],[2269,23.3577],[2270,22.7529],[2271,21.8016],[2272,20.565],[2273,19.1041],[2274,17.48],[2275,15.7538],[2277,12.2397],[2278,10.574],[2279,9.0507],[2280,7.7308],[2281,6.6755],[2282,5.9459],[2283,5.6031],[2284,5.4809],[2285,5.3719],[2286,5.2754],[2287,5.1904],[2289,5.0514],[2291,4.9463],[2293,4.8667],[2296,4.7831],[2298,4.75],[2300,4.7403],[2344,5.4321],[2354,5.4647],[2365,5.4486],[2391,5.6748],[2392,5.6576],[2393,5.6199],[2397,5.3748],[2398,5.3245],[2399,5.2941],[2421,4.8852],[2422,4.8603],[2425,4.7525],[2426,4.7258],[2427,4.7125],[2442,4.7668],[2444,4.7463],[2452,4.5751],[2455,4.5384],[2490,4.6827],[2494,4.7379],[2497,4.8083],[2500,4.9096],[2503,5.044],[2507,5.272],[2524,6.4717],[2530,6.8118],[2535,7.0179],[2542,7.192],[2583,7.2826],[2584,7.2471],[2585,7.1442],[2586,6.9783],[2587,6.7536],[2588,6.4741],[2589,6.1435],[2590,5.7661],[2591,5.3461],[2592,4.8885],[2593,4.3989],[2594,3.8832],[2595,3.3484],[2598,1.7041],[2599,1.1695],[2600,0.6554],[2601,0.17],[2602,-0.2791],[2603,-0.6842],[2604,-1.0383],[2605,-1.3345],[2606,-1.5663],[2607,-1.7272],[2608,-1.8108],[2609,-1.8216],[2820,-1.8216]]
       }
       foot_l_ik_ctrl.Rotation.Z {
        type=MovieSceneScriptingFloatChannel
        default=0
        thinned_from=1021
        keys [[1800,0],[1801,0.1897],[1802,0.7233],[1803,1.5475],[1804,2.6097],[1805,3.8569],[1806,5.2358],[1807,6.6923],[1808,8.1718],[1809,9.6193],[1810,10.9789],[1811,12.1946],[1812,13.2098],[1813,13.968],[1814,14.4127],[1815,14.4875],[1816,14.1364],[1817,13.3037],[1818,11.9345],[1819,9.9739],[1820,7.3677],[1821,4.0615],[1822,0],[1823,-4.9055],[1824,-10.5977],[1825,-16.9137],[1826,-23.6906],[1827,-30.7651],[1829,-45.1534],[1830,-52.1398],[1831,-58.7695],[1832,295.1217],[1833,289.6976],[1834,285.1219],[1835,281.5586],[1836,279.1716],[1837,278.1248],[1852,272.8715],[1856,268.6382],[1861,260.1809],[1873,233.0723],[1874,-129.216],[1878,-137.8319],[1892,-163.7696],[1899,-174.4706],[1905,-181.4302],[1963,-220.433],[1987,-227.0722],[2040,-226.7498],[2041,-225.1844],[2044,-216.7234],[2045,-214.5758],[2046,-213.7308],[2084,-228.33],[2250,-229.2496],[2251,-230.0611],[2252,-232.0734],[2253,-235.1411],[2254,-239.1182],[2255,-243.8585],[2256,110.7839],[2257,104.9539],[2261,79.7671],[2262,73.7063],[2263,68.0175],[2264,62.8388],[2265,58.3087],[2266,54.5661],[2267,51.7508],[2268,50.003],[2283,31.9951],[2284,30.5382],[2285,28.5572],[2286,26.1149],[2287,23.2738],[2288,20.0968],[2289,16.6464],[2290,12.9855],[2291,9.1768],[2292,5.2832],[2293,1.3673],[2294,-2.5079],[2295,-6.2796],[2296,-9.8849],[2297,-13.2609],[2298,-16.3448],[2299,-19.0736],[2300,-21.3845],[2301,-23.2146],[2302,-24.574],[2305,-28.3685],[2307,-30.7356],[2308,-31.8394],[2309,-32.8763],[2310,-33.836],[2311,-34.708],[2312,-35.4822],[2313,-36.1482],[2331,-44.9415],[2332,-45.5057],[2333,-46.2228],[2334,-47.0838],[2335,-48.0796],[2336,-49.2014],[2337,-50.4402],[2339,-53.2327],[2341,-56.3855],[2344,-61.633],[2345,-63.4844],[2346,294.6279],[2358,272.4201],[2363,265.6967],[2391,250.4325],[2392,248.6666],[2393,245.9286],[2397,231.4639],[2398,-131.4792],[2399,-133.5332],[2421,-166.6444],[2422,-168.4847],[2425,-175.7694],[2426,-177.8411],[2427,-179.2624],[2441,-186.3808],[2442,-187.0949],[2444,-189.6733],[2446,-193.4796],[2455,-215.9657],[2457,-219.9684],[2459,-222.8152],[2483,-243.9132],[2484,115.2097],[2488,111.6024],[2490,109.0866],[2492,105.8766],[2494,102.0419],[2496,97.6522],[2498,92.7772],[2500,87.4864],[2503,78.9229],[2507,66.6992],[2517,35.2554],[2520,26.021],[2522,19.9873],[2523,17.0178],[2524,14.0844],[2525,11.1907],[2526,8.3401],[2527,5.5362],[2528,2.7824],[2529,0.0823],[2530,-2.5606],[2531,-5.1428],[2532,-7.6608],[2533,-10.1111],[2534,-12.4902],[2535,-14.7945],[2536,-17.0206],[2537,-19.1649],[2538,-21.224],[2539,-23.1943],[2540,-25.0722],[2541,-26.8544],[2542,-28.5373],[2543,-30.1173],[2544,-31.5911],[2545,-32.955],[2546,-34.2056],[2547,-35.3394],[2548,-36.3529],[2549,-37.2426],[2550,-38.005],[2551,-38.6366],[2552,-39.1339],[2553,-39.4935],[2554,-39.7118],[2583,-39.7853],[2584,-39.3542],[2585,-38.1071],[2586,-36.1137],[2587,-33.4439],[2588,-30.1674],[2589,-26.3541],[2590,-22.0739],[2591,-17.3965],[2592,-12.3917],[2593,-7.129],[2594,-1.6778],[2595,3.8925],[2597,15.1145],[2598,20.6283],[2599,25.9856],[2600,31.1175],[2601,35.9552],[2602,40.4299],[2603,44.4725],[2604,48.014],[2605,50.9851],[2606,53.3166],[2607,54.9392],[2608,55.7838],[2820,55.8926]]
       }
       foot_l_ik_ctrl.Scale.X {type=MovieSceneScriptingFloatChannel, default=1, const [1,1021]}
       foot_l_ik_ctrl.Scale.Y {type=MovieSceneScriptingFloatChannel, default=1, const [1,1021]}
       foot_l_ik_ctrl.Scale.Z {type=MovieSceneScriptingFloatChannel, default=1, const [1,1021]}
       ball_l_ik_ctrl.Location.X {type=MovieSceneScriptingFloatChannel, default=0, const [0,1021]}
       ball_l_ik_ctrl.Location.Y {type=MovieSceneScriptingFloatChannel, default=0, const [0,1021]}
       ball_l_ik_ctrl.Location.Z {type=MovieSceneScriptingFloatChannel, default=0, const [0,1021]}
       ball_l_ik_ctrl.Rotation.X {type=MovieSceneScriptingFloatChannel, default=0, const [0,1021]}
       ball_l_ik_ctrl.Rotation.Y {type=MovieSceneScriptingFloatChannel, default=0, const [0,1021]}
       ball_l_ik_ctrl.Rotation.Z {type=MovieSceneScriptingFloatChannel, default=0, const [0,1021]}
       ball_l_ik_ctrl.Scale.X {type=MovieSceneScriptingFloatChannel, default=1, const [1,1021]}
       ball_l_ik_ctrl.Scale.Y {type=MovieSceneScriptingFloatChannel, default=1, const [1,1021]}
       ball_l_ik_ctrl.Scale.Z {type=MovieSceneScriptingFloatChannel, default=1, const [1,1021]}
       foot_bk1_l_ctrl.Location.X {type=MovieSceneScriptingFloatChannel, default=0, const [0,1021]}
       foot_bk1_l_ctrl.Location.Y {type=MovieSceneScriptingFloatChannel, default=0, const [0,1021]}
       foot_bk1_l_ctrl.Location.Z {type=MovieSceneScriptingFloatChannel, default=0, const [0,1021]}
       foot_bk1_l_ctrl.Rotation.X {type=MovieSceneScriptingFloatChannel, default=0, const [0,1021]}
       foot_bk1_l_ctrl.Rotation.Y {type=MovieSceneScriptingFloatChannel, default=0, const [0,1021]}
       foot_bk1_l_ctrl.Rotation.Z {type=MovieSceneScriptingFloatChannel, default=0, const [0,1021]}
       foot_bk1_l_ctrl.Scale.X {type=MovieSceneScriptingFloatChannel, default=1, const [1,1021]}
       foot_bk1_l_ctrl.Scale.Y {type=MovieSceneScriptingFloatChannel, default=1, const [1,1021]}
       foot_bk1_l_ctrl.Scale.Z {type=MovieSceneScriptingFloatChannel, default=1, const [1,1021]}
       foot_roll_l_ctrl.Location.X {type=MovieSceneScriptingFloatChannel, default=0, const [0,1021]}
       foot_roll_l_ctrl.Location.Y {type=MovieSceneScriptingFloatChannel, default=0, const [0,1021]}
       foot_roll_l_ctrl.Location.Z {type=MovieSceneScriptingFloatChannel, default=0, const [0,1021]}
       foot_roll_l_ctrl.Rotation.X {type=MovieSceneScriptingFloatChannel, default=0, const [0,1021]}
       foot_roll_l_ctrl.Rotation.Y {type=MovieSceneScriptingFloatChannel, default=0, const [0,1021]}
       foot_roll_l_ctrl.Rotation.Z {type=MovieSceneScriptingFloatChannel, default=0, const [0,1021]}
       foot_roll_l_ctrl.Scale.X {type=MovieSceneScriptingFloatChannel, default=1, const [1,1021]}
       foot_roll_l_ctrl.Scale.Y {type=MovieSceneScriptingFloatChannel, default=1, const [1,1021]}
       foot_roll_l_ctrl.Scale.Z {type=MovieSceneScriptingFloatChannel, default=1, const [1,1021]}
       roll_blend_l_ctrl {type=MovieSceneScriptingFloatChannel, default=1, const [1,1021]}
       tip_l_ctrl.Location.X {type=MovieSceneScriptingFloatChannel, default=0, const [0,1021]}
       tip_l_ctrl.Location.Y {type=MovieSceneScriptingFloatChannel, default=0, const [0,1021]}
       tip_l_ctrl.Location.Z {type=MovieSceneScriptingFloatChannel, default=0, const [0,1021]}
       tip_l_ctrl.Rotation.X {type=MovieSceneScriptingFloatChannel, default=0, const [0,1021]}
       tip_l_ctrl.Rotation.Y {type=MovieSceneScriptingFloatChannel, default=0, const [0,1021]}
       tip_l_ctrl.Rotation.Z {type=MovieSceneScriptingFloatChannel, default=0, const [0,1021]}
       tip_l_ctrl.Scale.X {type=MovieSceneScriptingFloatChannel, default=1, const [1,1021]}
       tip_l_ctrl.Scale.Y {type=MovieSceneScriptingFloatChannel, default=1, const [1,1021]}
       tip_l_ctrl.Scale.Z {type=MovieSceneScriptingFloatChannel, default=1, const [1,1021]}
       heel_l_ctrl.Location.X {type=MovieSceneScriptingFloatChannel, default=0, const [0,1021]}
       heel_l_ctrl.Location.Y {type=MovieSceneScriptingFloatChannel, default=0, const [0,1021]}
       heel_l_ctrl.Location.Z {type=MovieSceneScriptingFloatChannel, default=0, const [0,1021]}
       heel_l_ctrl.Rotation.X {type=MovieSceneScriptingFloatChannel, default=0, const [0,1021]}
       heel_l_ctrl.Rotation.Y {type=MovieSceneScriptingFloatChannel, default=0, const [0,1021]}
       heel_l_ctrl.Rotation.Z {type=MovieSceneScriptingFloatChannel, default=0, const [0,1021]}
       heel_l_ctrl.Scale.X {type=MovieSceneScriptingFloatChannel, default=1, const [1,1021]}
       heel_l_ctrl.Scale.Y {type=MovieSceneScriptingFloatChannel, default=1, const [1,1021]}
       heel_l_ctrl.Scale.Z {type=MovieSceneScriptingFloatChannel, default=1, const [1,1021]}
       leg_l_softness {type=MovieSceneScriptingFloatChannel, default=0, const [0,1021]}
       leg_l_pv_ik_ctrl.Location.X {
        type=MovieSceneScriptingFloatChannel
        default=0
        thinned_from=1021
        keys [[1800,0],[1968,0],[1969,0.0005],[1970,0],[2109,0],[2110,0.0005],[2111,0],[2201,0],[2202,-0.0404],[2203,0],[2229,0],[2230,0.0005],[2231,0],[2236,0],[2237,0.0005],[2238,0],[2245,0],[2246,0.0005],[2247,0],[2249,0],[2250,0.0005],[2251,0],[2255,0],[2256,0.0005],[2257,0],[2258,0],[2259,0.0002],[2260,0.0002],[2261,0],[2263,0],[2264,-0.0002],[2266,0.0002],[2267,0.0002],[2269,-0.0002],[2271,0.0002],[2272,-0.0002],[2273,0],[2274,0],[2275,-0.0005],[2276,-0.0005],[2277,0],[2281,0],[2282,0.0005],[2283,0.0005],[2284,0],[2286,0],[2287,-0.0005],[2288,0],[2293,0],[2294,0.0005],[2295,0],[2304,0],[2305,0.0005],[2306,0],[2307,0.0002],[2308,-0.0002],[2310,0.0002],[2311,0],[2318,0],[2319,0.0002],[2320,-0.0002],[2321,0],[2323,0],[2324,0.0002],[2325,0.0002],[2326,0],[2327,0],[2328,0.0002],[2329,0],[2331,0],[2332,-0.0002],[2333,0],[2334,0],[2335,-0.0002],[2336,0.0002],[2337,0],[2339,0],[2340,-0.0002],[2341,0],[2344,0],[2345,-0.0002],[2346,0],[2350,0],[2351,-0.0002],[2352,0],[2353,0],[2354,0.0002],[2355,0],[2362,0],[2363,-0.0002],[2364,0.0002],[2365,0],[2373,0],[2374,0.0002],[2375,0],[2376,0],[2377,0.0002],[2379,-0.0002],[2380,0],[2392,0],[2393,-0.0002],[2394,0.0001],[2395,-0.0001],[2396,0.0002],[2397,-0.0001],[2398,0.0001],[2399,-0.0001],[2400,-0.0001],[2401,0.0001],[2402,-0.0002],[2404,0.0002],[2405,0],[2406,0],[2407,0.0002],[2408,-0.0001],[2409,0.0002],[2410,0],[2413,-0.0001],[2415,0.0002],[2416,-0.0001],[2417,-0.0002],[2418,0],[2419,0],[2420,-0.0002],[2421,0.0001],[2422,-0.0001],[2423,-0.0001],[2424,0.0002],[2425,0],[2426,0],[2427,0.0002],[2431,0.0002],[2432,0],[2437,0],[2438,0.0002],[2439,0],[2451,0],[2452,0.0002],[2453,0],[2454,0.0007],[2455,0],[2456,0],[2457,-0.0002],[2459,0.0002],[2460,0.0002],[2461,-0.0002],[2462,0.0002],[2463,-0.0002],[2464,-0.0005],[2465,-0.0002],[2466,0.0002],[2472,0.0002],[2474,-0.0002],[2475,0.0005],[2476,-0.0002],[2477,0.0007],[2478,0],[2479,0.0002],[2480,0.0007],[2481,0.0002],[2482,0.0005],[2483,0.0005],[2484,0],[2485,-0.0002],[2486,0],[2487,0],[2488,-0.0005],[2489,0],[2491,0.0005],[2493,-0.0005],[2494,0.0002],[2495,0],[2496,0.0005],[2497,-0.0002],[2498,0.0005],[2499,0.0005],[2500,-0.0002],[2501,0.0007],[2502,-0.0002],[2503,0.0005],[2504,-0.0002],[2506,0.0002],[2507,0],[2508,0.0005],[2509,0],[2510,-0.0002],[2511,0.0002],[2512,0],[2513,0],[2514,0.0002],[2516,0.0002],[2517,0],[2518,0.0002],[2520,-0.0002],[2522,-0.0002],[2523,0.0002],[2524,-0.0002],[2525,-0.0005],[2526,0],[2528,0],[2529,0.0002],[2531,-0.0002],[2533,0.0002],[2534,0.0002],[2535,-0.0002],[2536,0],[2537,0],[2538,-0.0002],[2539,0],[2540,0],[2541,-0.0002],[2542,-0.0002],[2543,0],[2545,0],[2546,0.0002],[2547,0.0002],[2548,-0.0002],[2549,-0.0002],[2550,0.0002],[2551,0],[2553,0],[2554,-0.0002],[2555,-0.0002],[2556,0.0002],[2557,0],[2558,0],[2559,-0.0002],[2560,0],[2564,0],[2565,-0.0002],[2566,0],[2567,-0.0002],[2568,0],[2569,-0.0002],[2570,0],[2573,0],[2574,0.0002],[2575,0],[2576,0],[2577,-0.0002],[2578,0],[2581,0],[2582,-0.0002],[2583,0],[2584,0],[2585,-0.0002],[2586,0],[2587,-0.0002],[2588,0],[2589,-0.0002],[2590,0],[2605,0],[2606,0.0002],[2607,0.0002],[2608,0],[2758,0],[2759,0.0002],[2760,0],[2820,0]]
       }
       leg_l_pv_ik_ctrl.Location.Y {
        type=MovieSceneScriptingFloatChannel
        default=0
        thinned_from=1021
        keys [[1800,0],[1810,0],[1811,0.0001],[1834,0],[1835,-0.0002],[1836,0],[1844,0],[1845,-0.0002],[1846,0],[1848,0],[1849,-0.0002],[1850,0],[1852,0],[1853,-0.0002],[1854,0],[1857,0],[1858,-0.0002],[1859,0],[1875,0],[1877,0.0001],[1879,0],[1880,0.0001],[1915,0],[1916,-0.0002],[1917,0],[1918,0],[1919,-0.0002],[1920,0],[2069,0],[2070,0.0005],[2071,0],[2201,0],[2202,0.1124],[2203,0],[2214,0],[2215,-0.0005],[2216,0],[2244,0],[2245,-0.0005],[2246,0],[2255,0],[2256,0.0005],[2257,0],[2261,0],[2262,0.0002],[2263,0],[2264,0.0002],[2265,0],[2267,0],[2268,-0.0002],[2275,0],[2277,-0.0001],[2361,0],[2362,-0.0002],[2363,0],[2366,0],[2367,-0.0002],[2368,0],[2371,-0.0001],[2372,0.0001],[2378,-0.0002],[2380,0],[2382,-0.0001],[2384,0],[2386,-0.0001],[2388,0],[2389,-0.0002],[2390,0],[2408,0],[2409,0.0005],[2410,0],[2529,0],[2530,0.0005],[2531,0.0005],[2532,0],[2534,0],[2535,-0.0005],[2536,0],[2539,0],[2540,0.0005],[2541,0],[2551,0],[2552,0.0005],[2553,0],[2556,0],[2557,-0.0005],[2558,0],[2559,0],[2560,0.0005],[2561,0],[2581,0],[2582,-0.0005],[2583,0],[2585,0],[2586,0.0005],[2587,0],[2591,0],[2592,-0.0005],[2593,0],[2777,0],[2778,-0.0005],[2784,-0.0005],[2785,0],[2786,-0.0005],[2820,-0.0005]]
       }
       leg_l_pv_ik_ctrl.Location.Z {
        type=MovieSceneScriptingFloatChannel
        default=0
        thinned_from=1021
        keys [[1800,0.0002],[1801,-0.001],[1802,0.0007],[1803,-0.0002],[1804,-0.0002],[1805,0],[1806,-0.001],[1807,-0.0012],[1808,0.0006],[1809,0.0006],[1810,0.0002],[1811,0.0005],[1813,-0.0006],[1814,0.0004],[1815,-0.0001],[1816,0.0001],[1817,-0.001],[1818,0.0007],[1819,-0.001],[1820,0.001],[1821,-0.0009],[1822,0.0007],[1824,-0.0003],[1825,-0.0004],[1826,0],[1827,-0.0003],[1828,0.0001],[1829,0],[1830,0.0004],[1832,-0.0005],[1833,0.0002],[1834,0.0002],[1835,-0.0002],[1836,0.0002],[1837,0.0004],[1838,-0.0001],[1847,0],[1848,0.0001],[1849,0],[1854,0.0001],[1856,-0.0001],[1858,0.0002],[1859,0],[1861,-0.0001],[1862,0.0002],[1863,0],[1864,0],[1865,0.0002],[1866,0],[1867,0],[1868,0.0002],[1869,0.0002],[1870,-0.0002],[1871,0.0002],[1872,0],[1873,0.0002],[1874,-0.0002],[1875,0],[1876,-0.0002],[1877,-0.0002],[1878,0],[1883,0],[1884,0.0005],[1885,0],[1888,0],[1889,0.0005],[1890,0.0005],[1891,0],[1892,0],[1893,0.0005],[1894,0],[1895,0],[1896,0.0005],[1897,0],[1898,0.0005],[1899,0],[1902,0],[1903,0.0005],[1904,0],[1912,0],[1913,0.0005],[1914,0],[1924,0],[1925,0.0005],[1926,0],[1927,0],[1928,0.0005],[1929,0],[1938,0],[1939,-0.0005],[1941,0.0005],[1942,0],[1946,0],[1947,0.0005],[1948,0],[1950,0],[1951,0.0005],[1952,0],[1959,0],[1960,-0.0005],[1961,0],[1962,-0.0005],[1963,0],[1964,0],[1965,0.0005],[1966,0],[1967,0.0005],[1968,0],[1978,0],[1979,0.0005],[1980,0.0005],[1981,0],[1982,0],[1983,-0.0005],[1984,0],[1988,0],[1989,0.0005],[1990,0],[1994,0],[1995,-0.0005],[1996,0],[1997,-0.0005],[1998,0],[1999,0],[2000,0.0005],[2001,0],[2003,0],[2004,0.0005],[2005,0],[2008,0],[2009,0.0005],[2010,-0.0005],[2011,0],[2016,0],[2017,0.0005],[2018,0],[2020,0],[2021,0.0005],[2022,0],[2038,0],[2039,0.0005],[2040,0],[2044,0],[2045,-0.0005],[2046,0],[2047,-0.0005],[2048,0],[2053,0],[2054,-0.0002],[2055,0.0002],[2057,0.0002],[2058,0],[2060,-0.0001],[2062,0.0002],[2063,0.0002],[2064,-0.0001],[2067,0.0003],[2071,-0.0002],[2073,0],[2074,-0.0002],[2075,-0.0002],[2076,0.0001],[2077,-0.0002],[2079,0.0001],[2080,-0.0004],[2082,-0.0002],[2083,0.0002],[2084,-0.0002],[2085,-0.0002],[2086,0.0002],[2087,0],[2088,-0.0005],[2089,0],[2092,0],[2093,0.0005],[2094,0],[2098,0],[2099,0.0005],[2100,-0.0005],[2101,0],[2105,0],[2106,-0.0005],[2107,0],[2111,0],[2112,-0.0005],[2113,0],[2115,0],[2116,-0.0005],[2117,0],[2127,0],[2128,-0.0002],[2129,0],[2136,0.0001],[2137,-0.0001],[2138,0.0001],[2140,-0.0002],[2141,0.0002],[2142,-0.0002],[2143,-0.0001],[2144,0.0003],[2145,-0.0002],[2147,0.0001],[2148,-0.0001],[2149,0.0001],[2150,-0.0001],[2152,0.0004],[2153,0.0001],[2154,0.0002],[2155,0.0001],[2158,0.0001],[2159,0.0002],[2160,-0.0001],[2163,-0.0002],[2164,0.0001],[2165,-0.0001],[2166,0],[2167,-0.0002],[2168,-0.0001],[2169,0.0004],[2170,0],[2171,0.0005],[2172,0.0002],[2173,0.0002],[2174,0],[2175,0.0001],[2176,-0.0002],[2177,0],[2178,0],[2179,0.0002],[2180,-0.0002],[2181,0.0001],[2184,-0.0001],[2187,0],[2189,-0.0002],[2192,0],[2193,-0.0001],[2194,0.0002],[2195,0.0001],[2199,0.0002],[2200,-0.0001],[2201,0.0001],[2202,0.1039],[2203,-0.0002],[2204,0.0001],[2205,0.0001],[2206,-0.0002],[2207,-0.0002],[2208,0.0002],[2209,0.0003],[2210,-0.0004],[2212,0.0004],[2213,-0.0001],[2214,-0.0002],[2215,0.0002],[2216,-0.0004],[2218,0.0003],[2219,0.0004],[2220,-0.0004],[2221,0.0002],[2222,-0.0001],[2223,0.0004],[2226,0.0002],[2227,0],[2228,0],[2229,-0.0002],[2230,0.0002],[2231,-0.0002],[2232,0],[2233,-0.0005],[2234,-0.0005],[2235,0],[2236,0],[2237,0.0005],[2238,0],[2239,0],[2240,0.0005],[2241,-0.0005],[2242,0],[2244,0],[2245,0.001],[2247,-0.001],[2248,0],[2257,0],[2258,0.001],[2259,0],[2281,0],[2282,0.001],[2283,0],[2302,0],[2303,0.001],[2304,0],[2306,0],[2307,-0.0005],[2308,0],[2309,0],[2310,0.0005],[2311,0],[2313,0],[2314,-0.0005],[2315,0],[2316,-0.0005],[2317,0],[2318,0],[2319,0.0002],[2322,-0.0005],[2323,0.0002],[2324,0.0001],[2326,0.0004],[2328,0.0002],[2329,0.0003],[2330,-0.0003],[2331,0.0004],[2332,-0.0002],[2333,0.0002],[2335,-0.0001],[2337,0.0002],[2338,0],[2339,-0.0005],[2340,0.0002],[2341,-0.0002],[2342,0],[2350,0],[2351,-0.0005],[2352,0],[2383,0],[2384,-0.001],[2385,0],[2414,0],[2415,-0.001],[2416,-0.001],[2417,0],[2434,0],[2435,0.0005],[2436,0],[2437,0],[2438,0.001],[2439,0],[2440,0.001],[2441,0.0005],[2442,0.0005],[2443,0.001],[2444,-0.0005],[2445,0],[2447,-0.001],[2448,0.0005],[2449,0.0007],[2450,-0.0005],[2451,0.0005],[2452,0.0007],[2453,-0.0005],[2454,0.0005],[2455,-0.0005],[2456,0],[2457,-0.0006],[2458,0],[2459,0.0009],[2460,-0.0001],[2461,-0.0004],[2462,0.0001],[2464,-0.0008],[2465,-0.0006],[2466,0.0003],[2467,-0.0004],[2468,0.0004],[2469,0.0007],[2471,0.0002],[2472,0.0007],[2473,0],[2474,-0.0005],[2475,0.0005],[2476,0],[2477,0],[2478,0.0005],[2479,-0.0005],[2481,0.0005],[2482,0.0005],[2483,0.001],[2484,-0.0005],[2485,0],[2486,0.001],[2487,0],[2488,0],[2489,0.0005],[2490,0],[2507,0],[2508,-0.001],[2509,0],[2510,0],[2511,-0.001],[2512,0],[2516,0],[2517,-0.001],[2518,0],[2532,0],[2533,-0.001],[2534,0],[2535,0],[2536,0.001],[2537,0.001],[2538,0],[2539,0],[2540,-0.001],[2541,0],[2554,0],[2555,0.0005],[2556,0],[2557,0.0005],[2558,-0.0005],[2559,0],[2560,-0.0005],[2561,0],[2562,0],[2563,0.0005],[2564,0],[2571,0],[2572,-0.0005],[2573,0],[2574,0],[2575,-0.0005],[2576,0],[2578,0],[2579,0.0002],[2580,0],[2584,0],[2585,-0.0002],[2587,-0.0004],[2588,0],[2590,-0.0004],[2591,0.0002],[2593,0.0002],[2595,0],[2597,0.0004],[2598,-0.0001],[2599,0],[2601,-0.0002],[2602,0.0003],[2603,0],[2604,-0.0006],[2605,0.0004],[2606,-0.0001],[2607,-0.0001],[2608,-0.0004],[2609,-0.0001],[2756,-0.0001],[2757,0.0002],[2758,-0.0002],[2759,0.0002],[2760,-0.0001],[2761,0.0003],[2762,0],[2763,0.0002],[2764,-0.0005],[2765,-0.0002],[2766,-0.001],[2767,0.0005],[2768,-0.0007],[2769,0.001],[2770,0.0005],[2771,-0.0015],[2772,0.0015],[2773,-0.0015],[2774,-0.0005],[2775,0.0039],[2777,-0.0039],[2779,0],[2780,0.0049],[2781,-0.0039],[2782,-0.0039],[2783,0],[2784,0.0068],[2785,0.0088],[2786,0.0049],[2787,-0.0039],[2788,0.0088],[2789,0.002],[2791,0.002],[2792,-0.0059],[2793,0],[2794,0.0098],[2795,-0.0078],[2796,0.0137],[2797,-0.0078],[2798,0.0078],[2799,-0.0117],[2820,-0.0117]]
       }
       leg_l_pv_ik_ctrl.Rotation.X {
        type=MovieSceneScriptingFloatChannel
        default=0
        thinned_from=1021
        keys [[1800,0],[1823,0],[1824,0.0001],[1825,0],[1826,0.0001],[1828,-0.0001],[1829,0.0001],[1830,-0.0001],[1832,-0.0001],[1834,0.0001],[1835,-0.0001],[1837,0.0002],[1838,0],[1839,0.0001],[1842,-0.0002],[1843,0],[1844,-0.0002],[1850,-0.0002],[1852,0.0001],[1854,-0.0001],[1856,0.0002],[1858,-0.0002],[1859,0.0001],[1861,0.0001],[1862,-0.0001],[1870,0.0001],[1875,0],[1877,0.0001],[1879,-0.0001],[1906,0.0001],[1907,-0.0001],[1912,0],[1913,-0.0001],[1915,0.0002],[1916,-0.0002],[1918,0],[1919,-0.0001],[1920,0.0001],[1922,0],[1923,0.0001],[1935,-0.0001],[1936,0.0001],[1938,0],[1939,0.0001],[1941,0],[1942,0.0001],[1943,-0.0001],[1944,0.0001],[1947,0],[1948,0.0003],[1951,-0.0002],[1953,0.0001],[1956,0.0002],[1957,0],[1960,0.0002],[1961,-0.0001],[1963,0.0002],[1968,-0.0001],[1973,0.0001],[1974,-0.0002],[1976,0.0001],[1977,-0.0001],[1984,0.0002],[1985,0],[1988,0],[1989,-0.0002],[1991,0.0001],[1992,-0.0001],[1995,0.0003],[1996,0.0001],[1997,0.0003],[1999,-0.0001],[2000,-0.0002],[2001,0.0002],[2005,-0.0001],[2008,0.0002],[2013,0.0001],[2014,0.0002],[2015,-0.0001],[2016,-0.0002],[2017,0.0001],[2018,-0.0001],[2019,0.0002],[2020,-0.0002],[2021,-0.0003],[2022,0.0001],[2023,-0.0001],[2024,0.0001],[2028,0],[2029,0.0002],[2032,-0.0001],[2033,0],[2035,-0.0001],[2037,0],[2039,-0.0002],[2040,0],[2041,-0.0001],[2042,0.0001],[2048,0.0001],[2049,-0.0001],[2051,0.0001],[2052,0],[2054,0.0002],[2055,-0.0001],[2057,0],[2058,-0.0002],[2059,0.0001],[2061,0.0002],[2064,0],[2065,0.0001],[2066,-0.0002],[2067,-0.0001],[2068,-0.0002],[2069,0.0002],[2071,0.0002],[2072,-0.0001],[2073,-0.0002],[2074,0],[2080,0.0002],[2082,-0.0001],[2083,0.0001],[2086,0],[2087,0.0002],[2088,-0.0002],[2091,0],[2094,-0.0002],[2095,0.0001],[2096,-0.0002],[2098,0],[2099,-0.0001],[2101,0.0001],[2102,-0.0001],[2104,0.0002],[2105,-0.0001],[2106,0.0003],[2107,-0.0001],[2108,0.0001],[2111,-0.0002],[2113,0.0002],[2115,-0.0001],[2116,0.0002],[2117,-0.0001],[2118,0.0002],[2120,-0.0002],[2124,0.0001],[2126,-0.0001],[2128,0.0002],[2129,0],[2131,0.0001],[2132,0],[2133,0.0001],[2136,-0.0001],[2137,0.0001],[2138,-0.0001],[2142,-0.0001],[2143,0.0001],[2144,0],[2145,0.0001],[2146,0],[2147,0.0001],[2148,-0.0001],[2150,0],[2151,-0.0002],[2152,0],[2153,-0.0001],[2157,-0.0001],[2158,0.0001],[2159,-0.0001],[2162,0.0001],[2163,-0.0001],[2168,0.0001],[2173,-0.0001],[2178,0.0001],[2185,-0.0002],[2186,0],[2193,0],[2194,-0.0002],[2195,0.0001],[2199,-0.0001],[2200,0],[2201,-0.0001],[2202,-0.277],[2203,0.0001],[2204,0.0001],[2205,-0.0001],[2206,0.0003],[2212,-0.0002],[2213,0.0002],[2214,0.0001],[2215,-0.0003],[2216,0.0001],[2221,-0.0002],[2224,-0.0002],[2225,0],[2226,-0.0001],[2228,0.0001],[2230,-0.0001],[2236,0],[2237,-0.0001],[2238,0.0002],[2242,0.0001],[2243,-0.0001],[2245,-0.0002],[2247,0.0002],[2251,0.0001],[2252,-0.0003],[2253,-0.0002],[2254,0.0001],[2255,0],[2256,0.0003],[2257,-0.0001],[2258,0],[2259,-0.0002],[2260,0.0002],[2261,0],[2262,0.0002],[2263,-0.0002],[2264,0.0003],[2265,-0.0001],[2267,0],[2268,-0.0003],[2271,-0.0001],[2272,-0.0002],[2274,0],[2276,-0.0003],[2280,0.0001],[2282,-0.0002],[2283,0],[2285,0.0001],[2288,0],[2291,-0.0001],[2292,0.0001],[2293,-0.0001],[2295,0.0002],[2296,-0.0001],[2298,-0.0001],[2299,0.0002],[2300,0.0002],[2301,-0.0001],[2302,0.0001],[2303,-0.0001],[2307,0.0001],[2310,-0.0001],[2314,0.0001],[2319,-0.0001],[2321,-0.0001],[2322,0.0002],[2323,0],[2326,-0.0001],[2328,0.0002],[2329,0],[2330,0.0002],[2331,-0.0001],[2333,0.0001],[2334,-0.0001],[2337,-0.0001],[2339,0.0002],[2340,-0.0001],[2348,0.0001],[2349,-0.0001],[2350,0.0002],[2352,0],[2354,0.0002],[2356,-0.0001],[2358,0.0001],[2360,-0.0001],[2362,-0.0001],[2368,0],[2369,0.0002],[2371,-0.0002],[2372,0.0002],[2374,0],[2378,-0.0002],[2380,0.0001],[2381,-0.0001],[2382,-0.0001],[2383,0.0001],[2386,-0.0001],[2388,0.0001],[2389,-0.0001],[2390,0],[2391,-0.0001],[2392,0.0001],[2393,-0.0001],[2394,0.0001],[2395,-0.0001],[2397,0.0001],[2398,-0.0001],[2399,0.0001],[2402,0.0001],[2403,-0.0002],[2404,0.0001],[2406,0],[2407,0.0002],[2408,-0.0001],[2409,0.0005],[2411,-0.0003],[2412,0.0002],[2413,-0.0001],[2415,0.0002],[2416,0.0004],[2419,-0.0002],[2422,0.0001],[2423,0],[2424,0.0001],[2425,-0.0003],[2426,-0.0001],[2427,-0.0003],[2429,0.0002],[2430,-0.0002],[2431,0.0001],[2432,0],[2433,0.0001],[2434,-0.0002],[2437,0],[2438,-0.0005],[2439,-0.0001],[2440,-0.0002],[2441,0.0002],[2445,0],[2447,0.0004],[2448,-0.0003],[2450,-0.0001],[2452,-0.0002],[2453,0.0002],[2454,-0.0002],[2455,0.0002],[2457,0],[2458,0.0002],[2459,-0.0001],[2461,0.0001],[2463,0.0001],[2464,0.0004],[2465,-0.0001],[2466,-0.0002],[2467,0.0001],[2468,-0.0001],[2471,0.0001],[2472,-0.0003],[2473,0.0003],[2475,-0.0001],[2476,0],[2477,-0.0002],[2479,0.0004],[2480,0],[2483,-0.0002],[2484,0.0001],[2485,0],[2486,-0.0002],[2489,0.0001],[2492,0],[2493,0.0001],[2496,0.0001],[2497,-0.0001],[2498,0.0001],[2499,-0.0004],[2501,0.0001],[2502,-0.0004],[2504,0],[2507,0.0001],[2508,-0.0001],[2509,0.0002],[2510,-0.0001],[2511,0.0002],[2512,-0.0001],[2514,0.0002],[2515,-0.0002],[2517,0.0001],[2518,0],[2519,0.0003],[2520,-0.0002],[2523,0.0002],[2526,-0.0002],[2527,0.0002],[2528,-0.0002],[2529,-0.0001],[2530,0.0005],[2531,0.0004],[2532,0.0001],[2534,0.0001],[2535,-0.0004],[2539,-0.0001],[2540,0.0004],[2541,-0.0002],[2543,0.0001],[2545,-0.0002],[2546,0],[2547,-0.0002],[2548,0],[2549,0],[2550,-0.0002],[2551,-0.0001],[2552,0.0004],[2553,0],[2554,0.0001],[2555,-0.0003],[2556,0.0003],[2557,-0.0004],[2558,0.0001],[2559,0],[2560,0.0004],[2561,0],[2563,-0.0002],[2566,0.0001],[2569,-0.0001],[2570,0.0001],[2571,0],[2574,0],[2575,0.0001],[2577,-0.0001],[2580,0.0002],[2582,-0.0003],[2583,-0.0004],[2586,0.0004],[2588,-0.0001],[2589,-0.0001],[2590,0.0003],[2591,0],[2592,-0.0004],[2594,0.0001],[2595,-0.0001],[2597,-0.0002],[2598,0],[2599,-0.0001],[2600,0.0001],[2602,-0.0001],[2604,0.0001],[2605,-0.0002],[2607,0],[2758,0],[2759,-0.0001],[2762,0],[2763,-0.0001],[2770,-0.0001],[2771,0.0001],[2772,-0.0001],[2774,0.0001],[2775,-0.0002],[2776,-0.0002],[2777,0.0002],[2779,0],[2780,-0.0004],[2781,0.0001],[2783,0],[2785,-0.0006],[2786,-0.0003],[2787,0.0002],[2788,-0.0005],[2789,-0.0002],[2791,-0.0003],[2792,0.0004],[2794,-0.0005],[2795,0.0004],[2796,-0.0006],[2797,0.0005],[2798,-0.0004],[2799,0.0006],[2820,0.0006]]
       }
       leg_l_pv_ik_ctrl.Rotation.Y {
        type=MovieSceneScriptingFloatChannel
        default=0
        thinned_from=1021
        keys [[1800,0],[1813,0.0001],[1818,0],[1819,0.0001],[1820,-0.0001],[1821,0],[1822,-0.0001],[1831,0],[1832,0.0002],[1834,-0.0002],[1835,0.0001],[1836,-0.0002],[1838,-0.0001],[1839,0.0002],[1840,-0.0001],[1842,0],[1844,-0.0001],[1845,0.0001],[1846,-0.0001],[1848,0.0001],[1852,0],[1853,-0.0002],[1855,0],[1856,-0.0001],[1859,0],[1861,0],[1862,0.0001],[1865,0.0001],[1866,-0.0001],[1883,0],[1899,0],[1900,-0.0002],[1902,0],[1907,0],[1908,-0.0003],[1912,-0.0001],[1913,0.0001],[1915,0.0001],[1916,-0.0002],[1921,0],[1922,-0.0001],[1923,0.0001],[1924,-0.0001],[1925,0],[1926,-0.0002],[1933,0.0001],[1934,-0.0002],[1936,0.0001],[1938,-0.0002],[1939,0.0002],[1941,-0.0001],[1942,0.0001],[1943,-0.0002],[1944,0.0001],[1945,-0.0001],[1946,0.0001],[1950,-0.0001],[1951,0],[1952,-0.0001],[1953,0.0001],[1954,-0.0001],[1959,0.0003],[1961,-0.0001],[1963,0.0003],[1964,0],[1965,0.0001],[1969,-0.0003],[1970,0.0001],[1971,-0.0002],[1973,0.0001],[1975,-0.0002],[1976,0.0002],[1978,-0.0002],[1983,0.0002],[1984,-0.0001],[1985,0.0001],[1994,-0.0001],[1995,0.0002],[1996,-0.0001],[1997,0],[1998,-0.0001],[1999,0.0001],[2002,0.0002],[2004,0.0001],[2005,-0.0002],[2006,0.0002],[2007,0.0002],[2008,0],[2009,0.0002],[2010,-0.0001],[2012,0.0001],[2014,0],[2015,0.0001],[2018,-0.0002],[2019,0],[2021,-0.0001],[2023,0.0001],[2025,-0.0001],[2028,0.0002],[2031,-0.0001],[2033,0.0002],[2034,0],[2035,0.0001],[2036,-0.0001],[2037,0.0003],[2040,-0.0002],[2041,0.0001],[2043,-0.0001],[2049,0.0001],[2051,0],[2053,0],[2054,0.0002],[2055,-0.0001],[2057,-0.0001],[2058,0.0002],[2059,-0.0001],[2063,-0.0001],[2066,0.0001],[2067,0],[2071,-0.0002],[2072,0],[2075,-0.0002],[2076,0],[2077,-0.0002],[2079,0],[2080,-0.0002],[2082,-0.0001],[2083,0.0002],[2084,-0.0002],[2085,-0.0002],[2086,0.0002],[2087,0.0003],[2088,-0.0002],[2090,0.0001],[2093,0.0001],[2094,-0.0001],[2095,0.0001],[2096,-0.0003],[2098,0],[2100,-0.0002],[2101,0.0002],[2102,-0.0001],[2104,0.0002],[2105,-0.0001],[2106,0.0002],[2107,-0.0002],[2108,0.0001],[2110,-0.0002],[2112,0.0001],[2115,-0.0001],[2116,0],[2117,-0.0001],[2119,0.0001],[2122,0.0001],[2123,-0.0001],[2126,0.0001],[2128,-0.0001],[2129,0.0001],[2137,0],[2140,0.0002],[2141,0],[2142,0.0002],[2143,0],[2144,0],[2145,-0.0001],[2146,0],[2147,-0.0001],[2148,0],[2150,-0.0001],[2151,0.0001],[2152,-0.0001],[2156,0],[2157,0.0001],[2158,-0.0001],[2162,-0.0002],[2163,0.0001],[2165,-0.0001],[2167,0.0001],[2168,-0.0001],[2171,0.0001],[2174,-0.0002],[2175,0.0001],[2178,-0.0001],[2185,0.0002],[2188,0.0001],[2190,-0.0002],[2191,0],[2192,-0.0002],[2194,0],[2195,-0.0001],[2198,0.0001],[2201,0],[2202,-1.2115],[2203,-0.0001],[2204,-0.0001],[2205,0.0001],[2206,-0.0002],[2207,0.0001],[2209,-0.0003],[2210,0.0001],[2211,0.0001],[2212,-0.0003],[2214,0.0002],[2215,-0.0002],[2216,0.0003],[2217,0.0002],[2218,-0.0003],[2219,-0.0002],[2220,0.0002],[2221,-0.0001],[2222,0.0001],[2223,-0.0002],[2225,0],[2227,-0.0002],[2229,0.0002],[2230,-0.0003],[2231,0.0002],[2232,0.0003],[2234,0],[2236,0],[2237,-0.0004],[2238,0.0002],[2240,-0.0003],[2241,0.0001],[2242,-0.0001],[2243,-0.0001],[2246,-0.0003],[2247,0.0001],[2249,0],[2250,-0.0005],[2251,0],[2252,0.0002],[2253,-0.0001],[2255,0.0002],[2256,-0.0004],[2257,0],[2260,-0.0001],[2261,0.0003],[2262,-0.0002],[2269,0.0002],[2270,-0.0001],[2272,0],[2273,-0.0001],[2274,0.0002],[2275,0],[2277,0.0003],[2278,-0.0002],[2279,-0.0003],[2281,0],[2282,-0.0004],[2283,-0.0003],[2284,-0.0001],[2285,-0.0002],[2286,0.0003],[2287,0.0003],[2288,0],[2290,-0.0001],[2291,0.0002],[2294,-0.0004],[2295,0.0002],[2296,-0.0003],[2297,-0.0001],[2298,-0.0003],[2299,0.0001],[2300,0.0002],[2301,-0.0002],[2302,0.0001],[2303,0],[2304,-0.0004],[2305,-0.0003],[2306,0],[2308,0.0001],[2310,-0.0001],[2313,0],[2314,0.0003],[2315,-0.0001],[2317,0.0003],[2318,0.0002],[2319,-0.0003],[2320,0.0001],[2321,0],[2322,0.0001],[2323,-0.0002],[2325,-0.0001],[2326,-0.0003],[2327,0],[2331,-0.0001],[2332,0.0003],[2333,-0.0002],[2334,0.0001],[2335,0.0002],[2338,-0.0002],[2339,0.0003],[2340,0.0001],[2341,0.0001],[2342,0.0004],[2343,0.0001],[2344,0.0001],[2345,0.0003],[2346,0],[2349,-0.0001],[2350,0],[2351,0.0003],[2354,-0.0001],[2355,0],[2356,-0.0001],[2357,0.0001],[2359,0],[2360,0.0002],[2362,-0.0001],[2363,0.0001],[2364,-0.0002],[2365,-0.0001],[2367,-0.0001],[2368,0.0001],[2371,-0.0001],[2372,0.0002],[2374,-0.0002],[2375,0],[2377,-0.0002],[2378,0],[2384,0.0002],[2385,0],[2388,0.0002],[2390,0],[2393,0.0002],[2394,-0.0001],[2395,-0.0001],[2396,-0.0003],[2397,0.0001],[2398,-0.0002],[2399,0.0003],[2401,-0.0004],[2402,0.0004],[2403,-0.0001],[2404,0],[2406,-0.0002],[2407,-0.0001],[2408,0.0002],[2409,-0.0003],[2410,0.0001],[2414,0.0001],[2415,-0.0001],[2416,0.0002],[2417,0.0003],[2418,-0.0001],[2420,0.0001],[2421,-0.0002],[2426,0.0001],[2428,-0.0002],[2429,0],[2432,-0.0001],[2436,0.0002],[2438,0],[2443,-0.0002],[2444,0.0001],[2445,0],[2447,0.0002],[2449,-0.0003],[2450,0.0001],[2451,-0.0004],[2452,-0.0003],[2453,0.0002],[2454,-0.0002],[2455,0.0002],[2456,-0.0002],[2457,0.0005],[2459,-0.0003],[2461,0.0004],[2462,0],[2463,0.0005],[2464,0.0003],[2465,0.0004],[2466,-0.0003],[2467,0],[2468,-0.0001],[2469,-0.0006],[2471,-0.0003],[2472,-0.0005],[2474,0.0007],[2475,-0.0004],[2476,0],[2477,-0.0005],[2479,0.0002],[2480,-0.0006],[2481,-0.0009],[2482,-0.0004],[2483,-0.0005],[2485,0.0003],[2486,-0.0003],[2487,-0.0001],[2488,0.0006],[2489,0],[2490,0],[2491,-0.0009],[2492,0.0002],[2493,0.0003],[2494,-0.0002],[2495,0.0001],[2496,-0.0004],[2497,0.0003],[2498,-0.0004],[2499,-0.0003],[2500,0.0001],[2501,-0.0009],[2502,0.0005],[2503,-0.0004],[2504,0.0005],[2505,0],[2506,-0.0003],[2507,0.0002],[2508,-0.0004],[2510,0.0005],[2511,-0.0001],[2512,0.0001],[2513,-0.0001],[2514,-0.0004],[2515,0.0001],[2516,-0.0002],[2517,0.0001],[2518,-0.0003],[2519,0],[2520,-0.0001],[2521,0.0001],[2522,0.0001],[2523,-0.0003],[2524,0.0002],[2529,-0.0002],[2530,0.0002],[2531,0.0003],[2532,-0.0001],[2534,-0.0002],[2535,0.0002],[2538,0.0002],[2539,-0.0002],[2542,0.0002],[2544,0],[2545,0.0001],[2547,-0.0002],[2549,0.0001],[2550,-0.0003],[2551,-0.0001],[2552,-0.0002],[2554,0.0002],[2557,-0.0002],[2558,0.0002],[2562,-0.0001],[2564,0],[2565,0.0003],[2566,-0.0001],[2567,0.0001],[2568,-0.0001],[2573,0],[2574,-0.0002],[2576,-0.0001],[2577,0.0001],[2579,-0.0001],[2580,0.0001],[2581,0.0001],[2582,0.0003],[2583,0],[2585,0.0002],[2586,0],[2587,0.0001],[2589,0.0001],[2591,-0.0001],[2599,-0.0001],[2600,0],[2603,-0.0001],[2604,0],[2606,-0.0002],[2757,-0.0002],[2758,0.0001],[2759,-0.0002],[2760,0],[2765,0],[2766,0.0001],[2767,0],[2768,0.0002],[2770,0],[2771,0.0002],[2772,0],[2774,0.0002],[2775,-0.0003],[2776,-0.0003],[2777,0.0004],[2778,0.0001],[2779,0],[2780,-0.0005],[2781,0.0002],[2783,0.0001],[2785,-0.001],[2786,-0.0005],[2787,0.0004],[2788,-0.0008],[2789,-0.0002],[2791,-0.0005],[2792,0.0007],[2794,-0.0008],[2795,0.0006],[2796,-0.001],[2797,0.0008],[2798,-0.0007],[2799,0.0011],[2820,0.0011]]
       }
       leg_l_pv_ik_ctrl.Rotation.Z {type=MovieSceneScriptingFloatChannel, default=0, thinned_from=1021, keys [[1800,0],[2201,0],[2202,-1.1919],[2203,0],[2820,0]]}
       leg_l_pv_ik_ctrl.Scale.X {type=MovieSceneScriptingFloatChannel, default=1, const [1,1021]}
       leg_l_pv_ik_ctrl.Scale.Y {type=MovieSceneScriptingFloatChannel, default=1, const [1,1021]}
       leg_l_pv_ik_ctrl.Scale.Z {type=MovieSceneScriptingFloatChannel, default=1, const [1,1021]}
       thigh_r_fk_ctrl.Location.X {type=MovieSceneScriptingFloatChannel, default=0, const [0,15]}
       thigh_r_fk_ctrl.Location.Y {type=MovieSceneScriptingFloatChannel, default=0, const [0,15]}
       thigh_r_fk_ctrl.Location.Z {type=MovieSceneScriptingFloatChannel, default=0, const [0,15]}
       thigh_r_fk_ctrl.Rotation.X {
        type=MovieSceneScriptingFloatChannel
        default=10.0219
        thinned_from=15
        keys [[2162,10.0219],[2180,9.794],[2202,10.1334],[2208,4.0639],[2219,-0.8863],[2586,-0.7984],[2612,10.0219],[2734,10.0219],[2748,10.473],[2754,10.473],[2759,-1.1337]]
       }
       thigh_r_fk_ctrl.Rotation.Y {
        type=MovieSceneScriptingFloatChannel
        default=-3.188
        thinned_from=15
        keys [[2162,-3.188],[2180,-1.794],[2202,-0.0031],[2208,5.5662],[2219,-4.5283],[2586,-4.5058],[2612,-3.188],[2734,-3.188],[2748,0.907],[2754,0.907],[2759,-6.1282]]
       }
       thigh_r_fk_ctrl.Rotation.Z {
        type=MovieSceneScriptingFloatChannel
        default=-2.8149
        thinned_from=15
        keys [[2162,-2.8149],[2180,-2.6032],[2202,-15.8217],[2208,-29.8522],[2219,6.0899],[2586,6.0183],[2612,-2.8149],[2734,-2.8149],[2748,-25.1777],[2754,-26.7443],[2759,7.4302]]
       }
       thigh_r_fk_ctrl.Scale.X {type=MovieSceneScriptingFloatChannel, default=1, const [1,2]}
       thigh_r_fk_ctrl.Scale.Y {type=MovieSceneScriptingFloatChannel, default=1, const [1,2]}
       thigh_r_fk_ctrl.Scale.Z {type=MovieSceneScriptingFloatChannel, default=1, const [1,2]}
       calf_r_fk_ctrl.Location.X {type=MovieSceneScriptingFloatChannel, default=0, const [0,15]}
       calf_r_fk_ctrl.Location.Y {type=MovieSceneScriptingFloatChannel, default=0, const [0,15]}
       calf_r_fk_ctrl.Location.Z {type=MovieSceneScriptingFloatChannel, default=0, const [0,15]}
       calf_r_fk_ctrl.Rotation.X {type=MovieSceneScriptingFloatChannel, default=0, thinned_from=19, keys [[2162,0],[2208,0],[2211,0.2301],[2219,0],[2759,0]]}
       calf_r_fk_ctrl.Rotation.Y {type=MovieSceneScriptingFloatChannel, default=0, thinned_from=19, keys [[2162,0],[2208,0],[2211,-0.1294],[2219,0],[2759,0]]}
       calf_r_fk_ctrl.Rotation.Z {
        type=MovieSceneScriptingFloatChannel
        default=22.071
        thinned_from=20
        keys [[2162,22.071],[2180,22.071],[2202,-0.2803],[2208,25.2298],[2211,33.2554],[2219,0],[2591,0.1841],[2601,21.3083],[2611,10.7188],[2625,25.1139],[2638,22.071],[2734,22.071],[2748,60.0669],[2752,62.5925],[2754,62.5925],[2759,1.5475]]
       }
       calf_r_fk_ctrl.Scale.X {type=MovieSceneScriptingFloatChannel, default=1, const [1,2]}
       calf_r_fk_ctrl.Scale.Y {type=MovieSceneScriptingFloatChannel, default=1, const [1,2]}
       calf_r_fk_ctrl.Scale.Z {type=MovieSceneScriptingFloatChannel, default=1, const [1,2]}
       foot_r_fk_ctrl.Location.X {type=MovieSceneScriptingFloatChannel, default=0, const [0,15]}
       foot_r_fk_ctrl.Location.Y {type=MovieSceneScriptingFloatChannel, default=0, const [0,15]}
       foot_r_fk_ctrl.Location.Z {type=MovieSceneScriptingFloatChannel, default=0, const [0,15]}
       foot_r_fk_ctrl.Rotation.X {
        type=MovieSceneScriptingFloatChannel
        default=-2.8885
        thinned_from=15
        keys [[2162,-2.8885],[2202,-2.8885],[2219,-5.058],[2586,-5.0402],[2612,-2.8885],[2754,-2.8885],[2759,-5.0402]]
       }
       foot_r_fk_ctrl.Rotation.Y {
        type=MovieSceneScriptingFloatChannel
        default=-10.6402
        thinned_from=15
        keys [[2162,-10.6402],[2202,-10.6402],[2219,-8.538],[2586,-8.5553],[2612,-10.6402],[2754,-10.6402],[2759,-8.5553]]
       }
       foot_r_fk_ctrl.Rotation.Z {
        type=MovieSceneScriptingFloatChannel
        default=30.8604
        thinned_from=15
        keys [[2162,30.8604],[2202,30.8604],[2219,39.066],[2586,38.9985],[2612,30.8604],[2754,30.8604],[2759,38.9985]]
       }
       foot_r_fk_ctrl.Scale.X {type=MovieSceneScriptingFloatChannel, default=1, const [1,2]}
       foot_r_fk_ctrl.Scale.Y {type=MovieSceneScriptingFloatChannel, default=1, const [1,2]}
       foot_r_fk_ctrl.Scale.Z {type=MovieSceneScriptingFloatChannel, default=1, const [1,2]}
       ball_r_fk_ctrl.Location.X {type=MovieSceneScriptingFloatChannel, default=0, const [0,1021]}
       ball_r_fk_ctrl.Location.Y {type=MovieSceneScriptingFloatChannel, default=0, const [0,1021]}
       ball_r_fk_ctrl.Location.Z {type=MovieSceneScriptingFloatChannel, default=0, const [0,1021]}
       ball_r_fk_ctrl.Rotation.X {type=MovieSceneScriptingFloatChannel, default=0, const [0,1021]}
       ball_r_fk_ctrl.Rotation.Y {type=MovieSceneScriptingFloatChannel, default=0, const [0,1021]}
       ball_r_fk_ctrl.Rotation.Z {type=MovieSceneScriptingFloatChannel, default=0, const [0,1021]}
       ball_r_fk_ctrl.Scale.X {type=MovieSceneScriptingFloatChannel, default=1, const [1,1021]}
       ball_r_fk_ctrl.Scale.Y {type=MovieSceneScriptingFloatChannel, default=1, const [1,1021]}
       ball_r_fk_ctrl.Scale.Z {type=MovieSceneScriptingFloatChannel, default=1, const [1,1021]}
       foot_r_ik_ctrl.Location.X {
        type=MovieSceneScriptingFloatChannel
        default=0
        thinned_from=1021
        keys [[1800,0],[1801,16.941],[1802,34.5117],[1803,52.5406],[1804,70.8733],[1808,144.9955],[1809,163.2399],[1810,181.1292],[1811,198.4968],[1812,215.1578],[1813,230.9319],[1814,245.6739],[1815,259.3116],[1816,271.8825],[1820,317.3895],[1821,330.3116],[1822,345.263],[1827,424.754],[1837,573.8842],[1849,655.3793],[1852,671.9486],[1863,708.9883],[1865,718.9076],[1867,732.1183],[1869,749.5869],[1871,772.2678],[1872,785.853],[1873,801.0845],[1874,818.0718],[1876,857.3149],[1878,901.3164],[1881,973.9094],[1895,1347.5029],[1899,1443.2144],[1902,1505.4414],[1905,1556.8794],[1934,1904.0986],[1983,2248.9512],[2040,2385.4297],[2041,2379.8896],[2043,2352.5273],[2045,2311.4326],[2047,2257.2881],[2049,2189.3657],[2051,2109.2808],[2053,2018.7449],[2056,1867.0938],[2060,1644.2925],[2063,1469.2939],[2064,1415.9633],[2065,1369.2496],[2066,1331.316],[2069,1243.7236],[2070,1217.4281],[2071,1196.9324],[2072,1185.0811],[2073,1184.7112],[2074,1196.989],[2075,1220.7186],[2076,1254.5477],[2077,1297.1239],[2078,1347.0947],[2079,1403.1083],[2080,1463.8125],[2085,1790.3904],[2086,1850.8596],[2090,2063.0723],[2093,2207.3311],[2095,2290.8809],[2097,2360.3711],[2098,2388.791],[2099,2412.4277],[2100,2430.8545],[2101,2443.6372],[2102,2450.3398],[2107,2439.4683],[2116,2367.4404],[2117,2353.2573],[2118,2330.5078],[2119,2300.3062],[2120,2263.7656],[2122,2176.1184],[2127,1923.8396],[2131,1741.4889],[2133,1657.0775],[2135,1581.0046],[2136,1547.0952],[2137,1516.4678],[2138,1489.5212],[2139,1466.6577],[2140,1448.2834],[2141,1434.8083],[2142,1426.6445],[2156,1400.4324],[2204,1397.772],[2205,1395.5277],[2206,1389.157],[2208,1367.0645],[2210,1338.2622],[2212,1301.6154],[2214,1255.6038],[2216,1200.723],[2218,1137.4595],[2220,1066.3033],[2222,987.744],[2224,902.2683],[2226,810.3638],[2228,712.5178],[2230,609.2178],[2231,555.6748],[2232,500.9507],[2233,445.1069],[2234,388.2041],[2235,330.3032],[2236,271.4653],[2237,211.751],[2238,151.2212],[2239,89.937],[2240,27.9595],[2241,-34.6504],[2242,-97.8311],[2243,-161.5215],[2245,-290.1924],[2250,-616.5811],[2251,-686.8584],[2252,-765.7803],[2253,-852.2676],[2254,-945.2305],[2255,-1043.5889],[2256,-1146.2783],[2258,-1360.5146],[2262,-1796.5889],[2264,-2002.7617],[2265,-2099.3887],[2266,-2190.3418],[2267,-2274.584],[2278,-3099.1055],[2281,-3305.4121],[2283,-3429.9834],[2285,-3535.0635],[2287,-3614.3193],[2289,-3668.5996],[2291,-3698.749],[2293,-3705.5938],[2295,-3689.9033],[2297,-3652.3721],[2299,-3593.625],[2301,-3514.2295],[2312,-2977.9922],[2322,-2449.1882],[2326,-2258.7402],[2329,-2131.6802],[2332,-2018.4136],[2333,-1975.6974],[2334,-1928.454],[2335,-1876.8901],[2336,-1821.2125],[2337,-1761.6284],[2338,-1698.3446],[2339,-1631.5682],[2341,-1488.3634],[2343,-1333.6654],[2345,-1169.1213],[2347,-996.3728],[2349,-817.0564],[2351,-632.8037],[2353,-445.2407],[2357,-66.6714],[2358,27.5078],[2359,121.0986],[2360,213.8975],[2361,305.7031],[2362,396.3135],[2363,485.5264],[2364,573.1401],[2365,658.9512],[2366,742.7583],[2375,1462.1123],[2380,1845.5605],[2385,2207.3701],[2389,2475.6455],[2392,2661.2725],[2395,2829.7871],[2399,3028.3213],[2402,3155.9355],[2406,3297.0781],[2412,3461.124],[2435,3921.3262],[2441,3994.1885],[2445,4008.4761],[2449,3986.6353],[2452,3942.4126],[2455,3870.8945],[2458,3768.7834],[2462,3585.3914],[2468,3256.7219],[2480,2545.2566],[2484,2330.2734],[2488,2139.7832],[2489,2087.127],[2490,2029.6299],[2492,1901.9395],[2494,1760.3721],[2496,1608.583],[2503,1049.4873],[2504,972.1406],[2505,896.7832],[2506,823.8682],[2507,753.8447],[2508,687.165],[2509,624.2803],[2510,565.6396],[2511,511.6943],[2512,462.8945],[2513,419.0518],[2514,376.6143],[2515,334.9873],[2516,294.1982],[2517,254.2734],[2518,215.2402],[2519,177.125],[2520,139.958],[2521,103.7646],[2522,68.5752],[2523,34.416],[2524,1.3184],[2525,-30.6895],[2526,-61.5791],[2527,-91.3184],[2528,-119.8789],[2529,-147.2285],[2530,-173.3369],[2531,-198.1738],[2532,-221.7061],[2533,-243.9023],[2534,-264.7305],[2535,-284.1582],[2536,-302.1523],[2537,-318.6816],[2538,-333.7119],[2539,-347.21],[2540,-359.1445],[2541,-369.4805],[2542,-378.1855],[2543,-385.2256],[2544,-390.5684],[2545,-394.1816],[2546,-396.0303],[2547,-396.082],[2548,-394.3057],[2549,-390.666],[2550,-385.1328],[2551,-377.6719],[2552,-368.2529],[2553,-356.8428],[2554,-343.4102],[2555,-327.9248],[2556,-310.4141],[2557,-290.9854],[2558,-269.7236],[2559,-246.7178],[2560,-222.0522],[2561,-195.8145],[2562,-168.0908],[2563,-138.9683],[2564,-108.5322],[2565,-76.8696],[2566,-44.0674],[2567,-10.2119],[2568,24.6108],[2569,60.314],[2570,96.811],[2571,134.0156],[2572,171.8413],[2573,210.2012],[2575,288.1787],[2583,604.1685],[2586,718.0723],[2588,790.5623],[2590,859.3779],[2592,923.8456],[2594,983.4179],[2596,1037.5963],[2598,1085.8199],[2600,1127.4056],[2602,1161.5793],[2604,1187.5652],[2606,1204.6527],[2607,1209.6569],[2757,1212.2921],[2759,1206.9958],[2761,1195.0643],[2763,1177.0179],[2765,1153.3772],[2767,1124.6624],[2770,1073.2147],[2773,1013.2778],[2777,923.1924],[2790,598.0677],[2793,526.4215],[2795,481.3137],[2797,438.938],[2798,418.9375],[2799,409.2625],[2820,409.2625]]
       }
       foot_r_ik_ctrl.Location.Y {
        type=MovieSceneScriptingFloatChannel
        default=0
        thinned_from=1021
        keys [[1800,0],[1801,11.032],[1802,22.5137],[1803,35.1211],[1804,49.3629],[1805,65.5582],[1806,83.8298],[1807,104.1121],[1808,126.1724],[1809,149.6453],[1810,174.0763],[1812,223.8478],[1813,248.2817],[1814,271.9473],[1815,294.6443],[1816,316.3121],[1817,337.0315],[1821,415.9221],[1833,670.3005],[1837,749.5125],[1848,639.6462],[1850,618.0847],[1851,604.5654],[1852,586.0698],[1853,562.9839],[1854,535.6929],[1855,504.5786],[1856,470.0227],[1857,432.4043],[1858,392.1011],[1859,349.4893],[1860,304.9443],[1861,258.8411],[1862,211.5542],[1863,163.4592],[1865,66.355],[1866,18.1057],[1867,-29.4312],[1868,-75.8684],[1869,-120.8163],[1870,-163.8816],[1871,-204.6689],[1872,-242.7803],[1873,-277.8157],[1874,-309.374],[1875,-337.0533],[1876,-362.2245],[1877,-386.5181],[1878,-409.9213],[1879,-432.4221],[1881,-474.8278],[1883,-513.8594],[1885,-549.5114],[1887,-581.7783],[1889,-610.655],[1891,-636.1376],[1893,-658.2234],[1895,-676.912],[1897,-692.2048],[1899,-704.1057],[1901,-712.6201],[1903,-717.7559],[1905,-719.5212],[1940,-667.2405],[1991,-532.3125],[2001,-495.6538],[2012,-443.4629],[2031,-335.0337],[2039,-286.145],[2040,-278.7944],[2041,-263.2588],[2042,-238.9893],[2043,-207.166],[2044,-169.4585],[2045,-127.1592],[2046,-80.7363],[2047,-29.9253],[2048,25.3975],[2049,84.7734],[2050,147.7446],[2051,213.8521],[2052,282.6367],[2053,353.6401],[2054,426.4038],[2059,800.5913],[2061,946.2896],[2062,1013.7515],[2063,1074.1938],[2064,1125.5635],[2065,1166.1113],[2066,1194.1147],[2067,1209.1929],[2068,1219.5903],[2069,1227.0864],[2071,1234.5381],[2072,1235.0439],[2073,1232.5479],[2074,1220.397],[2075,1198.4858],[2076,1167.9512],[2077,1129.9292],[2078,1085.5571],[2079,1035.9707],[2080,982.3071],[2082,867.2915],[2084,749.5977],[2085,692.5864],[2086,638.3135],[2087,587.915],[2089,492.6367],[2090,446.2627],[2091,401.2617],[2092,358.0229],[2093,316.9312],[2094,278.3662],[2095,242.7036],[2096,210.3169],[2097,181.5786],[2098,156.8623],[2099,136.5444],[2100,121.0039],[2101,110.6245],[2102,105.7935],[2103,106.2568],[2104,108.3677],[2105,111.417],[2106,115.3179],[2107,119.9858],[2108,125.334],[2109,131.2778],[2110,137.7305],[2111,144.6064],[2113,159.2876],[2116,182.3462],[2117,194.2256],[2118,213.8413],[2119,240.1987],[2120,272.3027],[2121,309.1597],[2122,349.7764],[2123,393.1597],[2126,530.0088],[2127,574.564],[2128,616.9448],[2130,697.9458],[2132,775.2134],[2133,811.5918],[2134,846.0225],[2135,878.1787],[2136,907.7354],[2137,934.3677],[2138,957.7495],[2139,977.5518],[2140,993.4409],[2141,1005.0811],[2142,1012.1333],[2154,1033.0586],[2204,1033.9912],[2205,1036.2842],[2206,1042.8232],[2207,1053.0342],[2209,1080.3716],[2211,1114.4922],[2212,1135.061],[2213,1158.0854],[2215,1211.2397],[2217,1273.4331],[2219,1344.1367],[2221,1422.8174],[2223,1508.9521],[2225,1602.0186],[2228,1753.4722],[2232,1974.7979],[2238,2339.8452],[2251,3203.4878],[2253,3325.2446],[2255,3429.9854],[2257,3514.4771],[2259,3575.459],[2260,3596.0425],[2261,3609.3879],[2262,3614.9861],[2263,3612.3057],[2264,3600.7988],[2265,3579.9121],[2266,3549.0903],[2267,3507.7827],[2268,3455.4497],[2270,3331.0791],[2273,3119.4768],[2281,2501.6084],[2284,2281.6731],[2285,2200.0894],[2286,2111.9131],[2287,2017.873],[2288,1918.6978],[2289,1815.1118],[2291,1597.5874],[2297,913.3164],[2298,802.0225],[2299,693.3779],[2300,588.084],[2301,486.8398],[2302,389.7061],[2304,197.5625],[2305,102.583],[2306,8.7705],[2307,-83.5449],[2308,-174.0352],[2309,-262.3711],[2310,-348.2227],[2311,-431.2607],[2312,-511.1562],[2313,-587.5791],[2314,-660.1992],[2318,-940.8516],[2321,-1144.4785],[2323,-1274.1934],[2325,-1397.3584],[2327,-1512.4971],[2329,-1618.1328],[2331,-1712.791],[2332,-1755.1514],[2333,-1793.666],[2335,-1859.8096],[2337,-1912.5234],[2339,-1953.1162],[2341,-1982.8916],[2343,-2003.1602],[2346,-2018.5938],[2350,-2018.0864],[2362,-1970.5015],[2365,-1970.6956],[2376,-2016.3252],[2381,-2015.2844],[2384,-2001.3502],[2387,-1974.6876],[2390,-1933.0498],[2392,-1895.7694],[2394,-1850.0542],[2396,-1795.5596],[2398,-1731.6572],[2400,-1657.0945],[2402,-1569.8579],[2404,-1470.7954],[2406,-1361.1577],[2408,-1242.1924],[2410,-1115.1479],[2412,-981.2739],[2414,-841.8179],[2416,-698.0293],[2419,-476.9502],[2421,-327.7881],[2422,-252.1064],[2423,-174.4707],[2424,-94.9678],[2425,-13.6689],[2426,69.3584],[2427,154.0566],[2428,240.3311],[2429,327.876],[2430,416.5166],[2432,596.5332],[2442,1514.9248],[2446,1868.7812],[2449,2120.498],[2452,2356.6953],[2455,2574.0176],[2458,2769.1113],[2476,3827.8203],[2481,4082.4258],[2485,4251.2812],[2488,4350.8027],[2491,4413.7285],[2494,4440.2676],[2497,4434.7627],[2501,4385.0928],[2506,4269.4629],[2515,3970.2051],[2541,2914.0493],[2547,2706.0513],[2551,2589.1665],[2558,2424.0591],[2562,2307.6655],[2567,2133.2783],[2586,1359.9243],[2589,1245.4048],[2591,1173.7627],[2593,1107.0093],[2595,1045.958],[2597,991.2612],[2599,943.4756],[2600,922.3516],[2601,903.1792],[2602,886.0488],[2603,871.0557],[2604,858.3013],[2605,847.8906],[2606,839.9292],[2607,834.5215],[2608,831.77],[2757,831.1094],[2758,828.6426],[2759,823.7954],[2760,816.6572],[2761,807.3188],[2762,795.8687],[2763,782.3979],[2764,766.9951],[2765,749.7515],[2766,730.7554],[2767,710.0977],[2768,687.8682],[2769,664.1562],[2770,639.0518],[2772,585.0254],[2774,526.5073],[2776,464.2163],[2778,398.8706],[2780,331.1899],[2783,226.8613],[2787,86.2891],[2788,51.4829],[2789,16.9907],[2790,-17.0967],[2791,-50.6899],[2792,-83.6997],[2793,-116.0356],[2794,-147.6079],[2795,-178.3267],[2796,-208.1021],[2797,-236.8447],[2798,-264.4644],[2799,-277.8247],[2820,-277.8247]]
       }
       foot_r_ik_ctrl.Location.Z {
        type=MovieSceneScriptingFloatChannel
        default=0
        thinned_from=1021
        keys [[1800,0],[1801,114.834],[1802,232.873],[1803,353.8594],[1804,477.5527],[1805,603.7109],[1806,732.1055],[1807,862.5273],[1809,1128.7168],[1812,1539.2402],[1823,3088.5352],[1827,3624.6719],[1830,4004.0947],[1833,4357.6396],[1836,4679.5244],[1839,4965.2856],[1841,5132.5605],[1843,5278.9727],[1845,5402.8799],[1847,5502.6387],[1849,5576.6011],[1892,6520.3916],[1905,6637.7095],[2041,6418.8359],[2057,6710.4248],[2060,6730.958],[2061,6727.9717],[2062,6707.5298],[2064,6624.4844],[2067,6446.7373],[2068,6355.4512],[2071,6021.7881],[2072,5933.8057],[2078,5612.6533],[2081,5487.1187],[2084,5402.3198],[2086,5375.4336],[2089,5384.0698],[2092,5434.4727],[2098,5618.0078],[2115,6330.8955],[2123,6551.3003],[2127,6600.6206],[2130,6586.4893],[2134,6491.7612],[2149,6043.3525],[2190,5558.7876],[2205,5514.144],[2208,5539.0562],[2214,5662.7012],[2222,5930.8901],[2248,7084.5825],[2260,7463.0679],[2266,7566.1978],[2270,7572.2061],[2282,7396.2236],[2756,7675.6602],[2757,7659.4082],[2758,7531.1377],[2759,7279.0757],[2760,6907.8916],[2761,6422.2583],[2762,5826.8457],[2763,5126.3262],[2764,4325.3711],[2765,3428.6523],[2766,2440.8398],[2767,1366.6074],[2768,210.625],[2769,-1022.4375],[2770,-2327.9062],[2771,-3701.1133],[2772,-5137.3789],[2773,-6632.043],[2774,-8180.4219],[2775,-9777.8594],[2776,-11419.6719],[2777,-13101.1875],[2778,-14817.7422],[2780,-18337.2734],[2789,-34676.1719],[2792,-39912.25],[2795,-44833.0156],[2797,-47876.0625],[2798,-49312.3438],[2799,-50007.1094],[2820,-50007.1094]]
       }
       foot_r_ik_ctrl.Rotation.X {
        type=MovieSceneScriptingFloatChannel
        default=0
        thinned_from=1021
        keys [[1800,0],[1801,0.0613],[1802,0.2308],[1803,0.4848],[1804,0.797],[1805,1.1405],[1806,1.489],[1807,1.8183],[1808,2.1071],[1809,2.3385],[1810,2.5001],[1811,2.5849],[1812,-357.409],[1819,-358.9225],[1820,0.7528],[1821,0.4013],[1822,0],[1823,-0.6119],[1824,-1.5587],[1825,-2.8178],[1826,-4.3653],[1827,-6.1755],[1828,-8.2212],[1829,-10.474],[1830,-12.905],[1831,344.5148],[1851,284.49],[1854,273.2982],[1867,214.8127],[1870,202.7805],[1872,195.8062],[1874,189.8993],[1876,185.2564],[1877,183.4701],[1878,182.0732],[1894,170.8593],[1903,167.64],[1925,170.6744],[1980,190.4786],[1990,191.1617],[1993,189.7767],[2003,180.6994],[2006,179.0847],[2039,178.8531],[2040,180.6054],[2041,192.919],[2042,213.1892],[2043,236.8699],[2044,259.4147],[2045,276.2764],[2046,282.9075],[2049,281.8029],[2053,277.4879],[2060,265.386],[2061,262.2994],[2062,252.0832],[2063,236.3689],[2065,200.1608],[2066,185.5244],[2067,176.0889],[2068,167.6893],[2069,159.6842],[2070,152.3369],[2071,145.9108],[2072,140.6693],[2087,83.3584],[2088,79.1787],[2089,74.3378],[2090,68.942],[2091,63.0977],[2092,56.9114],[2093,50.4894],[2095,37.3639],[2096,30.8733],[2097,24.5725],[2098,18.568],[2099,12.9662],[2100,7.8735],[2101,3.3962],[2102,359.6409],[2117,315.3839],[2120,304.261],[2124,286.288],[2128,266.4409],[2130,255.2536],[2136,216.3881],[2137,210.4473],[2138,204.9926],[2139,200.1488],[2140,196.0406],[2141,192.793],[2142,190.5305],[2149,182.7872],[2153,179.7764],[2156,178.6929],[2204,178.5837],[2205,185.1157],[2206,201.9741],[2207,225.0518],[2208,250.2417],[2209,273.4365],[2210,290.5291],[2211,297.4123],[2241,302.9221],[2244,300.9881],[2252,288.3457],[2253,-73.3931],[2256,-78.4047],[2259,-82.8784],[2261,-85.416],[2263,-87.4871],[2265,-88.9839],[2266,-89.4797],[2267,-89.786],[2285,-89.4173],[2299,-87.4989],[2302,-87.4983],[2305,-88.1855],[2311,-90.4838],[2313,-90.9294],[2339,-91.1505],[2340,268.8313],[2471,272.6685],[2472,-87.2987],[2583,-87.4381],[2584,-87.8401],[2585,-89.0033],[2586,-90.8631],[2587,-93.3554],[2588,-96.4162],[2589,-99.9819],[2590,-103.989],[2591,-108.3742],[2592,-113.0743],[2594,-123.1659],[2599,-149.4366],[2601,-158.9838],[2602,-163.2822],[2603,-167.1726],[2604,-170.5859],[2605,-173.4533],[2606,-175.7057],[2607,-177.2744],[2608,-178.0912],[2820,-178.1965]]
       }
       foot_r_ik_ctrl.Rotation.Y {
        type=MovieSceneScriptingFloatChannel
        default=0
        thinned_from=1021
        keys [[1800,0],[1801,0.0076],[1802,0.031],[1803,0.0728],[1804,0.1364],[1805,0.2253],[1806,0.3409],[1807,0.4827],[1808,0.6473],[1809,0.8289],[1811,1.2078],[1812,1.3839],[1813,1.5355],[1814,1.6509],[1815,1.7186],[1816,1.7277],[1817,1.6676],[1818,1.5281],[1819,1.2993],[1820,0.9724],[1821,0.5401],[1822,0],[1823,-0.5808],[1824,-1.1356],[1825,-1.662],[1826,-2.1578],[1827,-2.6218],[1828,-3.054],[1829,-3.4567],[1830,-3.8341],[1831,-4.1926],[1833,-4.8883],[1834,-5.2462],[1835,-5.6256],[1836,-6.0372],[1837,-6.4906],[1838,-7.021],[1839,-7.6419],[1840,-8.3303],[1841,-9.0629],[1843,-10.568],[1844,-11.294],[1845,-11.9714],[1846,-12.5769],[1847,-13.0874],[1848,-13.4796],[1849,-13.7304],[1850,-13.8165],[1851,-13.7513],[1852,-13.5691],[1853,-13.2786],[1854,-12.8882],[1855,-12.4062],[1856,-11.8411],[1857,-11.2014],[1858,-10.4955],[1859,-9.7318],[1860,-8.9188],[1861,-8.0651],[1862,-7.179],[1863,-6.2691],[1866,-3.4828],[1867,-2.564],[1868,-1.6645],[1869,-0.7932],[1870,0.0412],[1871,0.83],[1872,1.5643],[1873,2.2354],[1874,2.8344],[1875,3.3525],[1876,3.781],[1877,4.1109],[1878,4.3335],[1879,4.4401],[1880,4.4771],[1881,4.4978],[1882,4.503],[1883,4.4934],[1884,4.4699],[1885,4.4332],[1886,4.3843],[1887,4.3238],[1888,4.2527],[1889,4.1718],[1890,4.0818],[1892,3.8783],[1894,3.6489],[1897,3.2714],[1902,2.6125],[1904,2.3595],[1905,2.2389],[1906,2.1235],[1908,1.9011],[1910,1.67],[1912,1.4304],[1914,1.184],[1916,0.9321],[1918,0.6763],[1923,0.0291],[1924,-0.1002],[1925,-0.2289],[1926,-0.3569],[1927,-0.484],[1928,-0.6101],[1930,-0.8582],[1932,-1.0997],[1934,-1.333],[1936,-1.5566],[1938,-1.7689],[1940,-1.9684],[1942,-2.1534],[1944,-2.3225],[1957,-3.3195],[1963,-3.7382],[1968,-4.0488],[1973,-4.3132],[1977,-4.4842],[1981,-4.6131],[1984,-4.6788],[1987,-4.7155],[1989,-4.7227],[1990,-4.6602],[1991,-4.4822],[1992,-4.2026],[1993,-3.8355],[1994,-3.3949],[1995,-2.8947],[1996,-2.349],[1997,-1.7717],[1998,-1.177],[1999,-0.5787],[2000,0.009],[2001,0.5722],[2002,1.0969],[2003,1.5691],[2004,1.9748],[2005,2.2999],[2006,2.5304],[2007,2.6524],[2008,2.6683],[2039,2.6683],[2040,2.7203],[2041,3.0858],[2042,3.6883],[2043,4.394],[2044,5.0716],[2045,5.5923],[2046,5.8288],[2053,6.1768],[2056,6.2885],[2059,6.3563],[2060,6.3667],[2061,6.3373],[2062,6.0894],[2063,5.6342],[2064,5.0179],[2065,4.2865],[2066,3.4858],[2067,2.662],[2068,1.861],[2069,1.1289],[2070,0.5116],[2071,0.0552],[2072,-0.1944],[2073,-0.2253],[2074,-0.2148],[2075,-0.1964],[2076,-0.1713],[2077,-0.1408],[2078,-0.106],[2079,-0.0682],[2080,-0.0285],[2081,0.0118],[2082,0.0516],[2083,0.0896],[2084,0.1246],[2085,0.1555],[2086,0.181],[2087,0.1999],[2091,0.2619],[2097,0.3591],[2155,1.3797],[2174,1.6602],[2189,1.8299],[2202,1.9281],[2213,1.9685],[2218,1.9717],[2219,1.9662],[2220,1.9556],[2221,1.9405],[2223,1.8988],[2225,1.8449],[2232,1.6199],[2234,1.5618],[2236,1.5139],[2237,1.4951],[2238,1.4802],[2239,1.47],[2240,1.4648],[2241,1.5194],[2242,1.9229],[2243,2.6377],[2244,3.5681],[2245,4.6182],[2246,5.6922],[2247,6.6943],[2248,7.5289],[2249,8.1],[2250,8.3118],[2251,8.0384],[2252,7.2583],[2253,6.0323],[2254,4.4199],[2255,2.4808],[2256,0.275],[2257,-2.1367],[2258,-4.692],[2259,-7.3274],[2260,-9.9775],[2261,-12.5757],[2262,-15.0541],[2263,-17.3441],[2264,-19.3765],[2265,-21.0824],[2266,-22.3929],[2267,-23.2397],[2268,-23.5549],[2269,-23.3577],[2270,-22.7529],[2271,-21.8016],[2272,-20.565],[2273,-19.1041],[2274,-17.48],[2275,-15.7538],[2277,-12.2397],[2278,-10.574],[2279,-9.0507],[2280,-7.7308],[2281,-6.6755],[2282,-5.9459],[2283,-5.6031],[2284,-5.4809],[2285,-5.3719],[2286,-5.2754],[2287,-5.1904],[2289,-5.0514],[2291,-4.9463],[2293,-4.8667],[2296,-4.7831],[2298,-4.75],[2300,-4.7403],[2344,-5.4321],[2354,-5.4647],[2365,-5.4486],[2391,-5.6748],[2392,-5.6576],[2393,-5.6199],[2397,-5.3748],[2398,-5.3245],[2399,-5.2941],[2421,-4.8852],[2422,-4.8603],[2425,-4.7525],[2426,-4.7258],[2427,-4.7125],[2442,-4.7668],[2444,-4.7463],[2452,-4.5751],[2455,-4.5384],[2490,-4.6827],[2494,-4.7379],[2497,-4.8083],[2500,-4.9096],[2503,-5.044],[2507,-5.272],[2524,-6.4717],[2530,-6.8118],[2535,-7.0179],[2542,-7.192],[2583,-7.2826],[2584,-7.2471],[2585,-7.1442],[2586,-6.9783],[2587,-6.7536],[2588,-6.4741],[2589,-6.1435],[2590,-5.7661],[2591,-5.3461],[2592,-4.8885],[2593,-4.3989],[2594,-3.8832],[2595,-3.3484],[2598,-1.7041],[2599,-1.1695],[2600,-0.6554],[2601,-0.17],[2602,0.2791],[2603,0.6842],[2604,1.0383],[2605,1.3345],[2606,1.5663],[2607,1.7272],[2608,1.8108],[2609,1.8216],[2820,1.8216]]
       }
       foot_r_ik_ctrl.Rotation.Z {
        type=MovieSceneScriptingFloatChannel
        default=0
        thinned_from=1021
        keys [[1800,0],[1801,-0.1897],[1802,-0.7233],[1803,-1.5475],[1804,-2.6097],[1805,-3.8569],[1806,-5.2358],[1807,-6.6923],[1808,-8.1718],[1809,-9.6193],[1810,-10.9789],[1811,-12.1946],[1812,-13.2098],[1813,-13.968],[1814,-14.4127],[1815,-14.4875],[1816,-14.1364],[1817,-13.3037],[1818,-11.9345],[1819,-9.9739],[1820,-7.3677],[1821,-4.0615],[1822,0],[1823,4.9055],[1824,10.5977],[1825,16.9137],[1826,23.6906],[1827,30.7651],[1829,45.1534],[1830,52.1398],[1831,58.7695],[1832,64.8783],[1833,70.3024],[1834,74.8781],[1835,78.4414],[1836,80.8284],[1837,81.8752],[1850,86.0911],[1851,86.511],[1852,87.1285],[1853,87.933],[1855,90.0603],[1857,92.8074],[1859,96.0888],[1862,101.8262],[1873,126.9277],[1874,-230.784],[1891,-197.9318],[1899,-185.5294],[1905,-178.5698],[1940,-151.727],[1970,-136.6822],[1984,-133.1749],[2039,-133.0275],[2040,-133.2502],[2041,-134.8156],[2042,-137.3934],[2044,-143.2766],[2045,-145.4242],[2046,-146.2692],[2052,-145.0821],[2078,-133.1315],[2085,-131.5506],[2230,-133.7213],[2250,-130.7504],[2251,-129.9389],[2252,-127.9266],[2253,-124.8589],[2254,-120.8818],[2255,-116.1415],[2256,-110.7839],[2257,-104.9539],[2261,-79.7671],[2262,-73.7063],[2263,-68.0175],[2264,-62.8388],[2265,-58.3087],[2266,-54.5661],[2267,-51.7508],[2268,-50.003],[2283,-31.9951],[2284,-30.5382],[2285,-28.5572],[2286,-26.1149],[2287,-23.2738],[2288,-20.0968],[2289,-16.6464],[2290,-12.9855],[2291,-9.1768],[2292,-5.2832],[2293,-1.3673],[2294,2.5079],[2295,6.2796],[2296,9.8849],[2297,13.2609],[2298,16.3448],[2299,19.0736],[2300,21.3845],[2301,23.2146],[2302,24.574],[2305,28.3685],[2307,30.7356],[2308,31.8394],[2309,32.8763],[2310,33.836],[2311,34.708],[2312,35.4822],[2313,36.1482],[2331,44.9415],[2332,45.5057],[2333,46.2228],[2334,47.0838],[2335,48.0796],[2336,49.2014],[2337,50.4402],[2339,53.2327],[2341,56.3855],[2344,61.6329],[2354,80.6992],[2357,85.958],[2360,90.5733],[2362,93.1732],[2364,95.308],[2366,96.9059],[2390,108.8981],[2391,109.5675],[2392,111.3334],[2393,114.0714],[2394,117.4711],[2396,125.0137],[2397,128.5361],[2398,-228.5208],[2399,-226.4668],[2421,-193.3556],[2422,-191.5153],[2425,-184.2306],[2426,-182.1589],[2427,-180.7376],[2441,-173.6192],[2442,-172.9051],[2443,-171.7923],[2445,-168.5541],[2447,-164.2714],[2454,-146.3545],[2456,-141.9114],[2457,-140.0316],[2458,-138.4408],[2459,-137.1848],[2488,-111.6024],[2490,-109.0866],[2492,-105.8766],[2494,-102.0419],[2496,-97.6522],[2498,-92.7772],[2500,-87.4864],[2503,-78.9229],[2507,-66.6992],[2517,-35.2554],[2520,-26.021],[2522,-19.9873],[2523,-17.0178],[2524,-14.0844],[2525,-11.1907],[2526,-8.3401],[2527,-5.5362],[2528,-2.7824],[2529,-0.0823],[2530,2.5606],[2531,5.1428],[2532,7.6608],[2533,10.1111],[2534,12.4902],[2535,14.7945],[2536,17.0206],[2537,19.1649],[2538,21.224],[2539,23.1943],[2540,25.0722],[2541,26.8544],[2542,28.5373],[2543,30.1173],[2544,31.5911],[2545,32.955],[2546,34.2056],[2547,35.3394],[2548,36.3529],[2549,37.2426],[2550,38.005],[2551,38.6366],[2552,39.1339],[2553,39.4935],[2554,39.7118],[2583,39.7853],[2584,39.3542],[2585,38.1071],[2586,36.1137],[2587,33.4439],[2588,30.1674],[2589,26.3541],[2590,22.0739],[2591,17.3965],[2592,12.3917],[2593,7.129],[2594,1.6778],[2595,-3.8925],[2597,-15.1145],[2598,-20.6283],[2599,-25.9856],[2600,-31.1175],[2601,-35.9552],[2602,-40.4299],[2603,-44.4725],[2604,-48.014],[2605,-50.9851],[2606,-53.3166],[2607,-54.9392],[2608,-55.7838],[2820,-55.8926]]
       }
       foot_r_ik_ctrl.Scale.X {type=MovieSceneScriptingFloatChannel, default=1, const [1,1021]}
       foot_r_ik_ctrl.Scale.Y {type=MovieSceneScriptingFloatChannel, default=1, const [1,1021]}
       foot_r_ik_ctrl.Scale.Z {type=MovieSceneScriptingFloatChannel, default=1, const [1,1021]}
       ball_r_ik_ctrl.Location.X {type=MovieSceneScriptingFloatChannel, default=0, const [0,1021]}
       ball_r_ik_ctrl.Location.Y {type=MovieSceneScriptingFloatChannel, default=0, const [0,1021]}
       ball_r_ik_ctrl.Location.Z {type=MovieSceneScriptingFloatChannel, default=0, const [0,1021]}
       ball_r_ik_ctrl.Rotation.X {type=MovieSceneScriptingFloatChannel, default=0, const [0,1021]}
       ball_r_ik_ctrl.Rotation.Y {type=MovieSceneScriptingFloatChannel, default=0, const [0,1021]}
       ball_r_ik_ctrl.Rotation.Z {type=MovieSceneScriptingFloatChannel, default=0, const [0,1021]}
       ball_r_ik_ctrl.Scale.X {type=MovieSceneScriptingFloatChannel, default=1, const [1,1021]}
       ball_r_ik_ctrl.Scale.Y {type=MovieSceneScriptingFloatChannel, default=1, const [1,1021]}
       ball_r_ik_ctrl.Scale.Z {type=MovieSceneScriptingFloatChannel, default=1, const [1,1021]}
       foot_bk1_r_ctrl.Location.X {type=MovieSceneScriptingFloatChannel, default=0, const [0,1021]}
       foot_bk1_r_ctrl.Location.Y {type=MovieSceneScriptingFloatChannel, default=0, const [0,1021]}
       foot_bk1_r_ctrl.Location.Z {type=MovieSceneScriptingFloatChannel, default=0, const [0,1021]}
       foot_bk1_r_ctrl.Rotation.X {type=MovieSceneScriptingFloatChannel, default=0, const [0,1021]}
       foot_bk1_r_ctrl.Rotation.Y {type=MovieSceneScriptingFloatChannel, default=0, const [0,1021]}
       foot_bk1_r_ctrl.Rotation.Z {type=MovieSceneScriptingFloatChannel, default=0, const [0,1021]}
       foot_bk1_r_ctrl.Scale.X {type=MovieSceneScriptingFloatChannel, default=1, const [1,1021]}
       foot_bk1_r_ctrl.Scale.Y {type=MovieSceneScriptingFloatChannel, default=1, const [1,1021]}
       foot_bk1_r_ctrl.Scale.Z {type=MovieSceneScriptingFloatChannel, default=1, const [1,1021]}
       foot_roll_r_ctrl.Location.X {type=MovieSceneScriptingFloatChannel, default=0, const [0,1021]}
       foot_roll_r_ctrl.Location.Y {type=MovieSceneScriptingFloatChannel, default=0, const [0,1021]}
       foot_roll_r_ctrl.Location.Z {type=MovieSceneScriptingFloatChannel, default=0, const [0,1021]}
       foot_roll_r_ctrl.Rotation.X {type=MovieSceneScriptingFloatChannel, default=0, const [0,1021]}
       foot_roll_r_ctrl.Rotation.Y {type=MovieSceneScriptingFloatChannel, default=0, const [0,1021]}
       foot_roll_r_ctrl.Rotation.Z {type=MovieSceneScriptingFloatChannel, default=0, const [0,1021]}
       foot_roll_r_ctrl.Scale.X {type=MovieSceneScriptingFloatChannel, default=1, const [1,1021]}
       foot_roll_r_ctrl.Scale.Y {type=MovieSceneScriptingFloatChannel, default=1, const [1,1021]}
       foot_roll_r_ctrl.Scale.Z {type=MovieSceneScriptingFloatChannel, default=1, const [1,1021]}
       roll_blend_r_ctrl {type=MovieSceneScriptingFloatChannel, default=1, const [1,1021]}
       tip_r_ctrl.Location.X {type=MovieSceneScriptingFloatChannel, default=0, const [0,1021]}
       tip_r_ctrl.Location.Y {type=MovieSceneScriptingFloatChannel, default=0, const [0,1021]}
       tip_r_ctrl.Location.Z {type=MovieSceneScriptingFloatChannel, default=0, const [0,1021]}
       tip_r_ctrl.Rotation.X {type=MovieSceneScriptingFloatChannel, default=0, const [0,1021]}
       tip_r_ctrl.Rotation.Y {type=MovieSceneScriptingFloatChannel, default=0, const [0,1021]}
       tip_r_ctrl.Rotation.Z {type=MovieSceneScriptingFloatChannel, default=0, const [0,1021]}
       tip_r_ctrl.Scale.X {type=MovieSceneScriptingFloatChannel, default=1, const [1,1021]}
       tip_r_ctrl.Scale.Y {type=MovieSceneScriptingFloatChannel, default=1, const [1,1021]}
       tip_r_ctrl.Scale.Z {type=MovieSceneScriptingFloatChannel, default=1, const [1,1021]}
       heel_r_ctrl.Location.X {type=MovieSceneScriptingFloatChannel, default=0, const [0,1021]}
       heel_r_ctrl.Location.Y {type=MovieSceneScriptingFloatChannel, default=0, const [0,1021]}
       heel_r_ctrl.Location.Z {type=MovieSceneScriptingFloatChannel, default=0, const [0,1021]}
       heel_r_ctrl.Rotation.X {type=MovieSceneScriptingFloatChannel, default=0, const [0,1021]}
       heel_r_ctrl.Rotation.Y {type=MovieSceneScriptingFloatChannel, default=0, const [0,1021]}
       heel_r_ctrl.Rotation.Z {type=MovieSceneScriptingFloatChannel, default=0, const [0,1021]}
       heel_r_ctrl.Scale.X {type=MovieSceneScriptingFloatChannel, default=1, const [1,1021]}
       heel_r_ctrl.Scale.Y {type=MovieSceneScriptingFloatChannel, default=1, const [1,1021]}
       heel_r_ctrl.Scale.Z {type=MovieSceneScriptingFloatChannel, default=1, const [1,1021]}
       leg_r_softness {type=MovieSceneScriptingFloatChannel, default=0, const [0,1021]}
       leg_r_pv_ik_ctrl.Location.X {
        type=MovieSceneScriptingFloatChannel
        default=0
        thinned_from=1021
        keys [[1800,0],[1958,0],[1959,0.0005],[1960,0],[2201,0],[2202,-0.043],[2203,0],[2236,0],[2237,-0.0005],[2238,0],[2243,0],[2244,-0.0002],[2246,-0.0002],[2247,0],[2249,0],[2250,-0.0002],[2251,0],[2255,0.0001],[2256,-0.0002],[2257,0],[2261,0.0001],[2262,-0.0001],[2264,0],[2265,-0.0002],[2268,-0.0001],[2269,0.0003],[2272,-0.0001],[2274,0.0002],[2275,0],[2276,0.0002],[2277,0.0002],[2278,-0.0004],[2280,0],[2282,-0.0001],[2283,0],[2284,-0.0002],[2285,-0.0001],[2286,0.0002],[2287,0.0002],[2288,-0.0001],[2289,0.0001],[2290,-0.0001],[2291,0.0001],[2292,0.0001],[2294,-0.0002],[2295,0.0002],[2296,-0.0002],[2297,0],[2298,-0.0002],[2300,0.0002],[2301,-0.0002],[2302,0],[2303,0],[2304,-0.0002],[2305,-0.0002],[2306,0],[2307,0],[2308,0.0002],[2309,0],[2310,0],[2311,0.0002],[2312,0],[2313,0],[2314,0.0002],[2315,0],[2317,0],[2318,0.0002],[2319,0],[2320,0.0002],[2321,0],[2322,0],[2323,-0.0002],[2324,-0.0002],[2325,0],[2326,-0.0002],[2327,0],[2328,-0.0002],[2330,0.0002],[2331,0],[2332,0.0002],[2333,-0.0002],[2335,0.0002],[2336,0],[2338,0],[2339,0.0002],[2340,0.0002],[2341,0],[2342,0.0002],[2343,0],[2344,0],[2345,0.0002],[2346,0],[2350,0],[2351,0.0002],[2352,0.0002],[2353,0],[2395,0],[2396,-0.0005],[2397,0],[2400,0],[2401,-0.0005],[2402,0.0005],[2403,0],[2408,0],[2409,-0.0005],[2410,0],[2415,0],[2416,0.0002],[2417,0.0002],[2418,0],[2419,0.0002],[2420,0.0002],[2421,-0.0002],[2422,0],[2426,0],[2427,0.0002],[2428,-0.0002],[2429,0],[2433,0],[2434,0.0002],[2435,0],[2441,0],[2442,-0.0002],[2443,0],[2444,-0.0002],[2445,0.0002],[2446,0],[2447,0.0002],[2448,0],[2451,0],[2452,-0.0002],[2453,0],[2454,0],[2455,0.0002],[2456,0],[2457,0.0002],[2458,0],[2459,-0.0005],[2460,0.0002],[2461,0.0002],[2462,0],[2463,0.0005],[2464,0.0002],[2465,0.0002],[2466,-0.0005],[2467,0],[2468,0],[2470,-0.0005],[2471,-0.0005],[2473,0],[2474,0.0005],[2475,-0.0004],[2476,0.0001],[2477,-0.0005],[2478,0.0001],[2479,0],[2480,-0.0006],[2481,-0.0004],[2482,-0.0005],[2484,-0.0002],[2485,0.0001],[2487,-0.0002],[2488,0.0004],[2489,0],[2490,0],[2491,-0.0006],[2492,0],[2493,0.0001],[2494,-0.0002],[2495,-0.0001],[2496,-0.0002],[2497,0.0001],[2498,-0.0004],[2500,0],[2501,-0.0007],[2502,0.0005],[2503,-0.0002],[2504,0.0004],[2506,-0.0002],[2507,0.0001],[2508,-0.0002],[2509,0.0001],[2510,0.0002],[2511,-0.0001],[2512,0.0001],[2514,-0.0002],[2515,0],[2516,-0.0002],[2517,0.0002],[2518,-0.0002],[2519,0],[2521,0.0001],[2523,-0.0002],[2524,0.0001],[2528,0],[2529,-0.0002],[2531,0.0001],[2532,-0.0001],[2534,-0.0001],[2535,0.0001],[2536,0],[2538,0.0001],[2539,-0.0001],[2540,-0.0001],[2541,0.0001],[2544,0],[2545,0.0001],[2546,-0.0001],[2549,0.0001],[2550,-0.0001],[2551,0],[2553,-0.0001],[2554,0.0001],[2556,-0.0002],[2558,0.0001],[2562,-0.0001],[2565,0.0001],[2566,-0.0001],[2567,0.0001],[2568,-0.0001],[2571,0.0001],[2573,0],[2574,-0.0002],[2575,0],[2577,0.0001],[2581,0],[2585,0.0001],[2586,0],[2589,0.0001],[2597,-0.0001],[2605,0],[2606,-0.0002],[2607,0],[2758,0],[2759,-0.0002],[2760,0],[2820,0]]
       }
       leg_r_pv_ik_ctrl.Location.Y {
        type=MovieSceneScriptingFloatChannel
        default=0
        thinned_from=1021
        keys [[1800,0],[1836,0],[1837,0.0002],[1838,0],[1839,0.0002],[1840,0],[1844,0],[1845,-0.0002],[1846,0],[1848,0],[1849,-0.0002],[1850,-0.0002],[1851,0],[1852,0],[1853,-0.0002],[1854,0],[1855,0],[1856,0.0002],[1857,0],[1860,0],[1861,0.0002],[1862,0],[1865,0],[1866,0.0002],[1867,0],[1875,0],[1876,0.0001],[2201,0],[2202,0.1061],[2203,0],[2261,0],[2262,0.0002],[2263,0],[2264,0.0002],[2265,-0.0002],[2266,0],[2267,0],[2268,-0.0004],[2269,0],[2270,-0.0001],[2271,0],[2272,-0.0002],[2274,0.0001],[2277,-0.0001],[2279,0],[2280,0.0002],[2281,0],[2361,0],[2362,-0.0002],[2363,0],[2368,0],[2369,0.0002],[2371,-0.0001],[2372,0.0001],[2378,-0.0001],[2380,0],[2382,-0.0001],[2384,0],[2386,-0.0001],[2393,0],[2394,0.0002],[2395,0],[2408,0],[2409,0.0005],[2410,0],[2530,0],[2531,0.0005],[2532,0],[2534,0],[2535,-0.0005],[2536,0],[2539,0],[2540,0.0005],[2541,0],[2551,0],[2552,0.0005],[2553,0],[2556,0],[2557,-0.0005],[2558,0],[2559,0],[2560,0.0005],[2561,0],[2581,0],[2582,-0.0005],[2583,-0.0005],[2584,0],[2585,0],[2586,0.0005],[2587,0],[2591,0],[2592,-0.0005],[2593,0],[2820,0]]
       }
       leg_r_pv_ik_ctrl.Location.Z {
        type=MovieSceneScriptingFloatChannel
        default=0
        thinned_from=1021
        keys [[1800,0.0002],[1801,-0.0005],[1802,0.0007],[1803,0],[1804,0],[1805,0.0002],[1806,-0.0005],[1807,-0.0007],[1808,0.0005],[1809,0.001],[1810,0.0007],[1811,0.0002],[1812,0.0002],[1813,-0.0007],[1814,0.0002],[1815,-0.0002],[1816,0.0005],[1817,-0.001],[1818,0.001],[1819,-0.0007],[1820,0.001],[1821,-0.001],[1822,0.0012],[1823,0.0002],[1824,-0.0005],[1825,-0.0002],[1826,-0.0002],[1827,-0.0005],[1829,0],[1830,0.0005],[1832,-0.0005],[1833,0.0002],[1834,0.0002],[1835,-0.0002],[1836,0.0002],[1837,0.0002],[1838,0],[1852,0],[1854,0.0001],[1856,-0.0001],[1858,0.0001],[1861,-0.0003],[1862,0],[1866,-0.0001],[1868,0.0001],[1870,-0.0001],[1871,0.0002],[1873,0.0001],[1874,-0.0002],[1875,0],[1876,-0.0002],[1877,-0.0002],[1878,0.0002],[1879,0],[1881,0],[1882,-0.0002],[1883,-0.0002],[1884,0.0005],[1885,-0.0002],[1886,-0.0002],[1887,0],[1888,0],[1889,0.0002],[1890,0],[1892,0],[1893,0.0005],[1894,0],[1897,0],[1898,0.0005],[1899,0],[1910,0],[1911,0.0005],[1912,0],[1927,0],[1928,0.0005],[1929,0],[1938,0],[1939,-0.0005],[1940,0],[1941,0],[1942,-0.0005],[1943,0],[1945,0],[1946,-0.0005],[1947,0],[1948,-0.0005],[1949,0],[1950,0],[1951,0.0005],[1952,0],[1958,0],[1959,-0.0005],[1960,-0.0005],[1961,0],[1962,0],[1963,-0.0005],[1965,0.0005],[1966,0],[1979,0],[1980,0.0005],[1981,0],[1982,0],[1983,-0.0005],[1984,0],[1988,0],[1989,0.0005],[1990,0],[1994,0],[1995,-0.0005],[1996,0],[1997,-0.0005],[1998,0],[1999,0],[2000,0.0005],[2001,0],[2003,0],[2004,0.0005],[2005,0],[2020,0],[2021,0.0005],[2022,0],[2038,0],[2039,0.0005],[2040,0],[2046,0],[2047,-0.0005],[2048,0],[2053,0],[2054,-0.0002],[2055,0.0002],[2057,0.0001],[2058,-0.0001],[2059,0],[2061,-0.0001],[2062,0.0002],[2063,0.0002],[2064,-0.0001],[2065,-0.0001],[2067,0.0002],[2069,-0.0001],[2072,-0.0001],[2073,0.0001],[2074,-0.0001],[2076,0],[2077,-0.0004],[2079,0.0001],[2080,-0.0004],[2082,-0.0002],[2083,0.0002],[2085,-0.0002],[2086,0.0001],[2087,0.0002],[2088,-0.0002],[2089,-0.0002],[2090,0],[2091,0],[2092,0.0002],[2093,0.0002],[2094,0],[2098,0],[2099,0.0005],[2100,0],[2101,0],[2102,0.0005],[2104,-0.0005],[2105,0],[2106,-0.0005],[2107,0],[2115,0],[2116,-0.0005],[2117,0.0005],[2118,0],[2119,0],[2120,0.0005],[2121,-0.0002],[2122,-0.0002],[2123,0],[2124,0],[2125,0.0002],[2126,-0.0002],[2127,0],[2128,-0.0002],[2130,0],[2131,-0.0001],[2132,0],[2133,-0.0001],[2134,0.0001],[2136,0.0001],[2137,-0.0001],[2138,0],[2140,-0.0002],[2141,0.0002],[2142,-0.0002],[2144,0.0002],[2146,-0.0002],[2147,0.0001],[2148,-0.0001],[2149,0.0001],[2151,0.0001],[2152,0.0003],[2154,0],[2159,0.0001],[2160,-0.0001],[2163,-0.0002],[2164,0.0001],[2165,-0.0003],[2166,-0.0001],[2168,0],[2169,0.0002],[2170,-0.0001],[2171,0.0004],[2173,0],[2174,0.0001],[2176,-0.0001],[2178,-0.0001],[2179,0.0002],[2180,-0.0001],[2183,0],[2184,-0.0002],[2185,-0.0001],[2186,-0.0002],[2187,0.0001],[2188,-0.0002],[2190,-0.0001],[2191,-0.0004],[2192,0.0001],[2193,-0.0001],[2194,0.0002],[2195,0],[2199,0.0001],[2200,-0.0001],[2201,0.0001],[2202,0.0973],[2203,-0.0001],[2205,0.0001],[2206,-0.0002],[2209,0.0004],[2210,-0.0004],[2212,0.0004],[2213,-0.0001],[2214,-0.0001],[2215,0.0002],[2216,-0.0004],[2217,-0.0001],[2218,0.0005],[2219,0.0005],[2220,-0.0005],[2221,0.0002],[2222,0],[2223,0.0002],[2226,0.0002],[2227,0],[2228,0],[2229,-0.0005],[2230,0],[2231,-0.0005],[2232,0],[2233,-0.0005],[2234,0],[2235,0],[2236,0.0005],[2237,0.0005],[2238,-0.0005],[2240,0.0005],[2241,-0.001],[2242,0],[2244,0],[2245,0.001],[2247,-0.001],[2248,0],[2257,0],[2258,0.001],[2259,0],[2302,0],[2303,0.001],[2304,0],[2306,0],[2307,-0.001],[2308,0],[2309,0],[2310,0.0005],[2311,0],[2313,0],[2314,-0.0005],[2315,0],[2316,-0.0005],[2317,-0.0005],[2319,0.0005],[2320,0],[2322,-0.0005],[2323,0.0002],[2324,0],[2326,0.0004],[2327,0.0001],[2329,0.0002],[2330,-0.0003],[2331,0.0003],[2332,-0.0003],[2333,0.0002],[2335,-0.0002],[2337,0.0002],[2338,0.0001],[2339,-0.0005],[2340,0.0001],[2341,-0.0001],[2344,-0.0002],[2345,0],[2349,0],[2350,-0.0002],[2351,-0.0002],[2352,0],[2354,0],[2355,-0.0005],[2356,0],[2383,0],[2384,-0.001],[2385,0],[2386,0],[2387,-0.001],[2388,0],[2415,0],[2416,-0.001],[2417,0],[2427,0],[2428,-0.001],[2429,0],[2431,0],[2432,-0.0005],[2433,0],[2434,0],[2435,0.0005],[2436,-0.0005],[2437,0],[2438,0.001],[2439,0],[2440,0.001],[2441,0.0005],[2442,0.0005],[2443,0.001],[2444,-0.0005],[2445,0],[2446,-0.0005],[2447,-0.0012],[2448,0.0002],[2449,0.0007],[2450,-0.0005],[2451,0.0005],[2452,0.0007],[2453,-0.0005],[2454,0.0005],[2455,-0.0005],[2456,0],[2457,-0.0006],[2458,-0.0001],[2459,0.0009],[2460,-0.0001],[2461,-0.0005],[2462,0.0001],[2464,-0.0009],[2465,-0.0007],[2466,0.0004],[2467,-0.0004],[2469,0.0007],[2471,0.0002],[2472,0.0007],[2473,0],[2474,-0.0005],[2475,0.0005],[2476,-0.0005],[2477,0.0005],[2478,0.0005],[2479,-0.0005],[2481,0.0005],[2482,0.0005],[2483,0.001],[2484,-0.0005],[2485,0],[2486,0.001],[2487,0],[2488,-0.0005],[2489,0],[2510,0],[2511,-0.001],[2512,0],[2519,0],[2520,0.001],[2521,0],[2535,0],[2536,0.001],[2537,0.001],[2538,0],[2539,0],[2540,-0.001],[2541,0],[2544,0],[2545,0.001],[2546,0],[2554,0],[2555,0.001],[2556,0],[2559,0],[2560,-0.0005],[2561,0],[2562,0],[2563,0.0005],[2564,0],[2567,0],[2568,0.0005],[2569,0],[2571,0],[2572,-0.0005],[2573,0],[2578,0],[2579,0.0002],[2580,0],[2582,0],[2583,0.0002],[2585,-0.0002],[2587,-0.0002],[2588,0],[2590,-0.0004],[2591,0.0002],[2593,0.0002],[2595,0],[2597,0.0004],[2598,-0.0001],[2599,0],[2600,-0.0002],[2601,0],[2602,0.0003],[2603,0],[2604,-0.0007],[2605,0.0004],[2608,-0.0004],[2609,0],[2756,0],[2757,0.0002],[2758,-0.0003],[2759,0.0002],[2760,-0.0002],[2761,0.0003],[2762,-0.0002],[2763,0.0002],[2764,-0.0005],[2765,0],[2766,-0.0006],[2767,0.0002],[2768,-0.0012],[2769,0],[2770,0.0007],[2771,-0.0022],[2772,0.0005],[2773,-0.0015],[2774,-0.0005],[2775,0.0029],[2776,0.001],[2777,-0.0054],[2778,-0.001],[2779,0.001],[2780,0.0049],[2781,-0.0049],[2783,-0.001],[2784,0.0068],[2785,0.0088],[2786,0.0049],[2787,-0.0029],[2788,0.0088],[2789,0.0029],[2790,0.0029],[2791,0.002],[2792,-0.0098],[2794,0.0098],[2795,-0.0078],[2796,0.0137],[2797,-0.0078],[2798,0.0078],[2799,-0.0137],[2820,-0.0137]]
       }
       leg_r_pv_ik_ctrl.Rotation.X {
        type=MovieSceneScriptingFloatChannel
        default=0
        thinned_from=1021
        keys [[1800,0],[1807,0],[1823,0],[1824,0.0001],[1825,0],[1826,0.0001],[1828,-0.0001],[1829,0.0001],[1832,-0.0001],[1834,0.0001],[1835,-0.0001],[1837,0.0002],[1838,0],[1839,0.0001],[1841,-0.0002],[1843,0.0001],[1845,-0.0001],[1846,0],[1847,-0.0001],[1850,-0.0002],[1851,0.0001],[1853,-0.0002],[1855,0.0001],[1858,-0.0002],[1859,0.0001],[1870,0.0001],[1875,0],[1877,0.0001],[1879,-0.0001],[1886,0],[1887,-0.0001],[1901,-0.0001],[1906,0.0001],[1911,-0.0001],[1915,0.0001],[1916,-0.0002],[1918,0],[1919,-0.0001],[1920,0.0001],[1932,0.0001],[1935,-0.0001],[1936,0.0001],[1938,0],[1939,0.0001],[1943,-0.0001],[1945,0.0001],[1947,0],[1948,0.0003],[1951,-0.0001],[1956,0.0001],[1957,0],[1960,0.0003],[1961,-0.0001],[1963,0.0002],[1967,-0.0002],[1972,0.0002],[1974,-0.0002],[1976,0.0001],[1977,-0.0001],[1981,0.0001],[1983,0.0001],[1984,0.0002],[1985,0],[1986,0.0002],[1989,-0.0002],[1991,0.0001],[1992,-0.0001],[1995,0.0002],[1996,0.0001],[1997,0.0003],[1999,-0.0001],[2000,-0.0002],[2001,0.0002],[2002,0],[2004,-0.0001],[2008,0.0002],[2010,0],[2014,0.0002],[2015,-0.0001],[2016,-0.0002],[2017,0.0001],[2018,-0.0001],[2019,0.0002],[2020,-0.0002],[2021,-0.0003],[2022,0.0001],[2023,-0.0001],[2024,0.0001],[2025,0],[2026,0.0001],[2028,0],[2029,0.0001],[2032,-0.0002],[2033,0],[2035,-0.0001],[2037,0],[2039,-0.0002],[2040,0],[2041,-0.0001],[2048,0.0001],[2052,0],[2054,0.0002],[2055,-0.0001],[2057,0],[2058,-0.0002],[2059,0.0002],[2061,0.0002],[2063,0],[2065,0.0001],[2066,-0.0002],[2067,0],[2068,-0.0002],[2069,0.0002],[2071,0.0002],[2072,-0.0002],[2073,-0.0002],[2074,0],[2079,0.0001],[2080,0.0003],[2082,-0.0001],[2090,0],[2094,-0.0001],[2095,0.0002],[2096,-0.0002],[2098,0],[2099,-0.0002],[2101,0.0002],[2102,-0.0001],[2104,0.0002],[2105,-0.0001],[2106,0.0002],[2107,-0.0001],[2108,0.0002],[2110,-0.0002],[2113,0.0002],[2115,-0.0001],[2116,0.0002],[2117,-0.0001],[2118,0.0002],[2120,-0.0002],[2124,0.0001],[2127,0],[2128,0.0002],[2129,0],[2131,0.0001],[2132,-0.0001],[2133,0.0001],[2136,-0.0002],[2137,0.0001],[2138,-0.0001],[2142,-0.0001],[2143,0.0001],[2144,0],[2145,0.0001],[2146,0],[2147,0.0001],[2148,-0.0001],[2150,0],[2151,-0.0002],[2152,0],[2153,-0.0001],[2157,-0.0001],[2158,0.0001],[2159,-0.0001],[2162,0.0001],[2163,-0.0001],[2168,0.0001],[2173,-0.0001],[2178,0.0001],[2185,-0.0002],[2186,0],[2193,0],[2194,-0.0001],[2195,0.0001],[2199,-0.0001],[2200,0],[2201,-0.0001],[2202,-0.2679],[2203,0.0001],[2204,0.0001],[2205,-0.0001],[2206,0.0002],[2209,0],[2210,0],[2211,-0.0002],[2212,-0.0001],[2213,0.0002],[2215,-0.0002],[2217,0.0002],[2219,-0.0001],[2220,0],[2224,-0.0002],[2226,-0.0001],[2227,0.0001],[2230,-0.0001],[2231,0],[2237,-0.0001],[2238,0.0002],[2240,0],[2242,0.0001],[2243,-0.0001],[2245,-0.0002],[2247,0.0002],[2251,0],[2252,-0.0003],[2254,0],[2255,0],[2256,0.0004],[2257,-0.0001],[2258,0],[2259,-0.0002],[2260,0.0002],[2261,0],[2262,0.0002],[2263,-0.0002],[2264,0.0004],[2265,-0.0001],[2267,0],[2268,-0.0004],[2269,-0.0002],[2270,-0.0003],[2271,-0.0001],[2272,-0.0002],[2274,0.0001],[2276,-0.0003],[2277,-0.0003],[2279,0.0001],[2282,-0.0002],[2283,0],[2285,0.0001],[2287,-0.0001],[2291,-0.0001],[2292,0.0001],[2293,-0.0001],[2295,0.0002],[2296,-0.0001],[2298,-0.0001],[2299,0.0002],[2300,0.0002],[2301,-0.0001],[2302,0.0001],[2303,-0.0001],[2307,0.0001],[2310,-0.0002],[2314,0.0001],[2315,0],[2316,0.0001],[2319,-0.0002],[2321,-0.0001],[2322,0.0002],[2323,0],[2327,0],[2328,0.0001],[2329,0],[2330,0.0002],[2331,-0.0001],[2332,-0.0001],[2333,0.0001],[2334,-0.0001],[2335,0],[2337,-0.0001],[2339,0.0002],[2340,-0.0001],[2348,0.0001],[2349,-0.0001],[2350,0.0002],[2352,0],[2354,0.0002],[2357,-0.0001],[2358,0.0001],[2362,-0.0001],[2368,-0.0001],[2369,0.0002],[2371,-0.0002],[2372,0.0002],[2378,-0.0002],[2380,0.0001],[2382,-0.0001],[2383,0.0001],[2386,-0.0001],[2387,0.0001],[2389,-0.0001],[2390,0],[2391,-0.0001],[2392,0.0001],[2393,-0.0001],[2394,0.0001],[2395,-0.0001],[2397,0.0002],[2398,-0.0001],[2399,0.0001],[2402,0.0001],[2403,-0.0002],[2404,0.0001],[2405,0],[2407,0.0001],[2408,-0.0001],[2409,0.0005],[2411,-0.0003],[2412,0.0002],[2413,-0.0001],[2416,0.0004],[2419,-0.0002],[2422,0.0001],[2423,0],[2424,0.0001],[2425,-0.0003],[2426,-0.0001],[2427,-0.0003],[2428,0.0001],[2429,0.0002],[2430,-0.0002],[2431,0.0001],[2432,0],[2433,0.0001],[2434,-0.0002],[2437,0],[2438,-0.0005],[2439,-0.0001],[2440,-0.0002],[2441,0.0001],[2445,0],[2447,0.0004],[2448,-0.0003],[2450,0],[2452,-0.0002],[2453,0.0002],[2454,-0.0002],[2455,0.0002],[2457,-0.0001],[2458,0.0002],[2459,-0.0001],[2463,0.0001],[2464,0.0004],[2465,-0.0001],[2466,-0.0002],[2467,0.0001],[2468,-0.0001],[2471,0.0001],[2472,-0.0003],[2473,0.0002],[2475,-0.0001],[2476,0],[2477,-0.0002],[2479,0.0004],[2480,0],[2483,-0.0002],[2484,0.0001],[2485,0.0001],[2486,-0.0002],[2490,0.0001],[2491,-0.0001],[2495,0.0002],[2497,-0.0001],[2498,0.0002],[2499,-0.0005],[2501,0.0001],[2502,-0.0004],[2504,0],[2507,0],[2508,-0.0001],[2509,0.0002],[2510,-0.0001],[2511,0.0002],[2512,-0.0001],[2514,0.0002],[2515,-0.0002],[2517,0.0001],[2518,-0.0001],[2519,0.0003],[2520,-0.0002],[2523,0.0002],[2526,-0.0002],[2527,0.0002],[2528,-0.0002],[2529,-0.0001],[2530,0.0005],[2531,0.0004],[2532,0.0001],[2534,0.0001],[2535,-0.0004],[2539,-0.0001],[2540,0.0004],[2541,-0.0002],[2543,0.0001],[2545,-0.0001],[2546,0],[2547,-0.0002],[2548,0.0001],[2550,-0.0002],[2551,-0.0001],[2552,0.0005],[2553,0],[2554,0.0001],[2555,-0.0002],[2556,0.0002],[2557,-0.0004],[2558,0.0001],[2559,0],[2560,0.0004],[2561,0],[2563,-0.0001],[2566,0.0001],[2569,-0.0002],[2570,0.0001],[2574,-0.0001],[2575,0.0001],[2577,-0.0001],[2580,0.0002],[2582,-0.0004],[2583,-0.0003],[2586,0.0004],[2588,-0.0001],[2589,-0.0001],[2590,0.0003],[2592,-0.0004],[2594,0.0001],[2597,-0.0002],[2598,0],[2599,-0.0002],[2600,0.0001],[2601,-0.0001],[2604,0.0001],[2605,-0.0002],[2758,0.0001],[2759,-0.0001],[2762,0],[2763,-0.0001],[2770,0],[2771,0.0001],[2772,0],[2774,0.0001],[2775,-0.0002],[2776,-0.0001],[2777,0.0003],[2779,0],[2780,-0.0003],[2781,0.0002],[2783,0],[2785,-0.0005],[2786,-0.0003],[2787,0.0002],[2788,-0.0005],[2789,-0.0001],[2791,-0.0002],[2792,0.0005],[2794,-0.0005],[2795,0.0004],[2796,-0.0006],[2797,0.0005],[2798,-0.0004],[2799,0.0007],[2820,0.0007]]
       }
       leg_r_pv_ik_ctrl.Rotation.Y {
        type=MovieSceneScriptingFloatChannel
        default=0
        thinned_from=1021
        keys [[1800,-0.0001],[1807,0.0001],[1820,-0.0001],[1821,0.0001],[1833,0],[1834,0.0001],[1835,-0.0001],[1836,0.0002],[1837,0],[1838,0.0001],[1839,-0.0002],[1840,0.0001],[1844,0.0002],[1845,-0.0001],[1847,0.0001],[1848,-0.0001],[1852,0],[1853,0.0001],[1855,0],[1861,0.0001],[1862,-0.0001],[1864,0],[1865,-0.0001],[1866,0.0001],[1879,0],[1899,0],[1900,0.0002],[1902,-0.0001],[1907,0],[1908,0.0002],[1910,0.0001],[1912,0.0001],[1913,-0.0002],[1915,0],[1916,0.0002],[1921,-0.0001],[1922,0.0001],[1923,-0.0001],[1924,0.0001],[1925,-0.0001],[1926,0.0002],[1930,-0.0001],[1932,0],[1933,-0.0001],[1935,0.0001],[1936,-0.0001],[1938,0.0002],[1939,-0.0002],[1941,0.0002],[1942,-0.0001],[1943,0.0002],[1944,-0.0001],[1945,0.0001],[1947,-0.0001],[1950,0.0001],[1951,-0.0001],[1952,0],[1953,-0.0001],[1954,0.0001],[1960,-0.0001],[1961,0.0001],[1963,-0.0002],[1964,0],[1965,-0.0001],[1966,0.0001],[1967,0],[1968,0.0003],[1969,0.0003],[1970,-0.0002],[1971,0.0001],[1973,-0.0001],[1975,0.0002],[1976,-0.0002],[1978,0.0002],[1981,0],[1982,-0.0002],[1984,0.0001],[1985,-0.0001],[1992,-0.0001],[1994,0.0001],[1995,-0.0002],[1996,0.0001],[1999,-0.0002],[2001,-0.0001],[2002,-0.0002],[2004,-0.0001],[2005,0.0001],[2006,-0.0002],[2007,-0.0003],[2008,0],[2009,-0.0003],[2010,0],[2011,-0.0001],[2013,0],[2016,-0.0002],[2018,0.0001],[2021,0],[2022,-0.0002],[2025,0],[2027,0],[2028,-0.0002],[2031,0.0001],[2033,-0.0002],[2034,0],[2035,-0.0001],[2036,0.0001],[2037,-0.0002],[2040,0.0002],[2041,-0.0001],[2042,0.0001],[2049,-0.0001],[2057,0.0001],[2058,-0.0002],[2059,0],[2063,0.0001],[2066,-0.0001],[2068,-0.0001],[2071,0.0002],[2072,0],[2075,0.0001],[2076,0],[2077,0.0002],[2079,0],[2080,0.0003],[2081,0.0001],[2082,0.0002],[2083,-0.0002],[2084,0.0002],[2085,0.0003],[2086,-0.0002],[2087,-0.0003],[2088,0.0002],[2089,0.0001],[2090,-0.0002],[2092,0],[2093,-0.0002],[2094,0.0001],[2095,-0.0002],[2096,0.0002],[2098,0],[2100,0.0002],[2101,-0.0002],[2102,0],[2104,-0.0001],[2105,0.0001],[2106,-0.0002],[2107,0.0003],[2108,-0.0001],[2110,0.0003],[2113,-0.0001],[2115,0.0001],[2116,0],[2117,0.0001],[2122,-0.0001],[2123,0.0001],[2128,0],[2129,-0.0001],[2135,0],[2136,-0.0001],[2137,0.0001],[2140,-0.0001],[2141,0],[2142,-0.0002],[2143,0],[2144,-0.0001],[2145,0.0001],[2150,0.0001],[2151,-0.0001],[2152,0],[2153,-0.0001],[2158,0],[2159,-0.0002],[2162,0.0002],[2163,-0.0001],[2165,0.0001],[2166,-0.0001],[2168,0.0002],[2169,0],[2174,0.0001],[2177,0],[2178,0.0001],[2185,-0.0001],[2190,0.0002],[2191,0],[2192,0.0002],[2194,-0.0001],[2197,0],[2198,-0.0001],[2201,0],[2202,-1.2116],[2203,0],[2205,-0.0001],[2206,0.0002],[2207,0],[2209,0.0002],[2211,-0.0001],[2212,0.0003],[2214,-0.0002],[2215,0.0002],[2216,-0.0002],[2217,-0.0001],[2218,0.0002],[2219,0.0002],[2220,-0.0002],[2221,0.0001],[2222,0],[2224,0.0002],[2225,0],[2227,0.0003],[2229,-0.0002],[2230,0.0003],[2231,-0.0001],[2232,-0.0003],[2234,0.0001],[2236,0],[2237,0.0003],[2238,-0.0002],[2240,0.0002],[2241,0],[2242,0.0001],[2245,0.0001],[2246,0.0003],[2247,-0.0001],[2249,0],[2250,0.0004],[2252,-0.0002],[2253,0],[2254,-0.0001],[2255,-0.0001],[2256,0.0004],[2257,0],[2258,-0.0002],[2259,0.0001],[2260,0.0001],[2261,-0.0002],[2263,0.0001],[2264,-0.0002],[2266,0.0003],[2267,0],[2269,-0.0002],[2270,0.0001],[2273,0.0001],[2274,-0.0003],[2275,0],[2276,-0.0002],[2278,0.0001],[2281,0.0001],[2282,0.0003],[2285,0],[2286,-0.0003],[2287,-0.0004],[2288,0],[2290,0.0002],[2291,-0.0003],[2294,0.0004],[2295,-0.0003],[2296,0.0003],[2297,0],[2298,0.0002],[2300,-0.0002],[2301,0.0001],[2303,-0.0002],[2304,0.0003],[2305,0.0003],[2307,0.0001],[2308,-0.0003],[2310,0.0002],[2311,-0.0001],[2313,0.0001],[2314,-0.0002],[2315,0],[2318,-0.0002],[2319,0.0002],[2320,-0.0002],[2321,0],[2322,-0.0001],[2323,0.0001],[2325,0],[2326,0.0002],[2327,0],[2331,0.0001],[2332,-0.0001],[2333,0],[2335,-0.0002],[2337,0.0001],[2338,0.0001],[2339,-0.0002],[2341,0],[2342,-0.0003],[2344,0],[2345,-0.0002],[2347,0.0001],[2349,0.0002],[2351,-0.0003],[2354,0.0001],[2356,0.0001],[2359,0],[2360,-0.0002],[2361,0.0001],[2362,0.0001],[2363,-0.0002],[2364,0.0002],[2365,0.0001],[2367,0.0001],[2368,-0.0001],[2371,0.0001],[2372,-0.0002],[2374,0.0002],[2375,0],[2377,0.0002],[2393,-0.0001],[2395,0.0001],[2396,0.0003],[2397,-0.0002],[2398,0.0002],[2399,-0.0003],[2401,0.0002],[2402,-0.0003],[2403,0],[2407,0.0001],[2408,-0.0002],[2409,0.0002],[2410,-0.0001],[2412,0],[2414,-0.0001],[2415,0.0002],[2417,-0.0003],[2418,0],[2419,0],[2420,-0.0002],[2421,0.0002],[2422,-0.0001],[2424,0.0002],[2426,-0.0001],[2428,0.0003],[2429,0],[2432,0.0001],[2433,-0.0001],[2434,0],[2436,-0.0001],[2438,0],[2439,-0.0001],[2441,0.0001],[2442,0],[2443,0.0002],[2444,0],[2445,0],[2446,-0.0002],[2449,0.0001],[2450,-0.0001],[2451,0.0003],[2452,0.0003],[2453,-0.0003],[2454,0.0005],[2455,-0.0002],[2456,0.0001],[2457,-0.0003],[2458,0.0001],[2459,0.0002],[2461,-0.0005],[2462,-0.0001],[2464,-0.0005],[2465,-0.0004],[2466,0.0003],[2468,-0.0001],[2469,0.0005],[2471,0.0003],[2472,0.0004],[2473,0.0001],[2474,-0.0005],[2475,0.0004],[2476,-0.0002],[2477,0.0006],[2478,0],[2479,0],[2480,0.0006],[2481,0.0009],[2482,0.0005],[2483,0.0006],[2485,-0.0006],[2486,0.0003],[2487,0],[2488,-0.0007],[2489,0.0001],[2490,0.0001],[2491,0.0008],[2492,-0.0003],[2493,-0.0002],[2494,0.0002],[2495,-0.0002],[2496,0.0003],[2497,-0.0003],[2498,0.0003],[2499,0.0004],[2500,-0.0001],[2501,0.0007],[2502,-0.0005],[2503,0.0005],[2504,-0.0004],[2506,0.0003],[2507,-0.0001],[2508,0.0003],[2510,-0.0006],[2511,0.0002],[2512,-0.0001],[2514,0.0004],[2515,0],[2516,0.0003],[2517,0],[2518,0.0002],[2521,-0.0002],[2522,0],[2523,0.0004],[2524,-0.0003],[2525,-0.0003],[2528,0.0002],[2529,0.0001],[2530,-0.0002],[2531,-0.0003],[2533,0.0003],[2534,0.0002],[2535,-0.0001],[2538,-0.0002],[2539,0.0001],[2540,0.0002],[2541,-0.0002],[2542,-0.0002],[2545,-0.0001],[2547,0.0003],[2549,-0.0002],[2550,0.0003],[2551,0.0001],[2553,0.0001],[2554,-0.0002],[2555,-0.0001],[2556,0.0002],[2557,0.0001],[2558,-0.0001],[2562,0.0001],[2563,-0.0001],[2564,0],[2565,-0.0003],[2566,0.0001],[2567,-0.0001],[2568,0.0001],[2572,0.0001],[2573,-0.0001],[2574,0.0002],[2578,-0.0001],[2579,0.0001],[2580,-0.0001],[2581,0],[2582,-0.0002],[2583,0.0001],[2585,-0.0002],[2586,0],[2587,-0.0001],[2589,-0.0001],[2590,0.0001],[2594,0],[2599,0.0001],[2600,0],[2603,0.0001],[2757,0.0001],[2758,-0.0001],[2759,0.0002],[2772,-0.0002],[2773,0.0001],[2774,-0.0001],[2778,0.0001],[2780,-0.0001],[2781,0.0003],[2783,0.0002],[2785,-0.0004],[2787,0.0003],[2788,-0.0003],[2789,-0.0001],[2791,-0.0001],[2792,0.0003],[2794,-0.0003],[2795,0.0003],[2796,-0.0007],[2797,0.0003],[2798,-0.0004],[2799,0.0006],[2820,0.0006]]
       }
       leg_r_pv_ik_ctrl.Rotation.Z {type=MovieSceneScriptingFloatChannel, default=0, thinned_from=1021, keys [[1800,0],[2201,0],[2202,-1.191],[2203,0],[2820,0]]}
       leg_r_pv_ik_ctrl.Scale.X {type=MovieSceneScriptingFloatChannel, default=1, const [1,1021]}
       leg_r_pv_ik_ctrl.Scale.Y {type=MovieSceneScriptingFloatChannel, default=1, const [1,1021]}
       leg_r_pv_ik_ctrl.Scale.Z {type=MovieSceneScriptingFloatChannel, default=1, const [1,1021]}
       upperarm_l_fk_ctrl.Location.X {type=MovieSceneScriptingFloatChannel, default=0, const [0,6]}
       upperarm_l_fk_ctrl.Location.Y {type=MovieSceneScriptingFloatChannel, default=0, const [0,6]}
       upperarm_l_fk_ctrl.Location.Z {type=MovieSceneScriptingFloatChannel, default=0, const [0,6]}
       upperarm_l_fk_ctrl.Rotation.X {
        type=MovieSceneScriptingFloatChannel
        default=1.1349
        keys [[2162,1.1349],[2180,1.1349],[2202,-3.4706],[2208,-0.0664],[2211,1.1493],[2220,1.6426],[2227,1.6426],[2240,1.5025],[2259,7.9802],[2278,-2.6489],[2296,-6.6124],[2306,-5.2122],[2313,-4.6064],[2333,-2.0453],[2351,-5.7971],[2374,-5.9509],[2383,-5.8565],[2414,-8.9114],[2432,-8.5972],[2441,-8.2977],[2454,-7.1826],[2477,-5.8027],[2493,-5.5198],[2515,-4.6177],[2583,1.1305],[2587,1.6384],[2593,15.1955],[2603,44.7319],[2616,47.9638],[2634,47.5011],[2644,46.1228],[2656,27.8429],[2672,-3.6213],[2704,5.6672],[2722,29.961],[2734,29.961],[2749,10.8172],[2754,10.8172],[2759,2.4224]]
       }
       upperarm_l_fk_ctrl.Rotation.Y {
        type=MovieSceneScriptingFloatChannel
        default=-4.7121
        keys [[2162,-4.7121],[2180,-4.7121],[2202,0.5339],[2208,5.8381],[2211,18.1111],[2220,-4.5603],[2227,-4.5603],[2240,-1.8446],[2259,4.938],[2278,6.7127],[2296,-0.2482],[2306,0.4643],[2313,0.1181],[2333,7.504],[2351,6.3048],[2374,1.7471],[2383,1.6419],[2414,7.3196],[2432,1.7432],[2441,0.9932],[2454,-0.1395],[2477,-0.6703],[2493,-0.8963],[2515,-1.8418],[2583,-4.5455],[2587,-4.5616],[2593,2.9227],[2603,19.0924],[2616,0.2033],[2634,7.5299],[2644,7.5392],[2656,7.7912],[2672,-3.7547],[2705,-0.0022],[2722,17.3123],[2734,16.2634],[2749,-2.2751],[2754,-2.2751],[2759,-11.9828]]
       }
       upperarm_l_fk_ctrl.Rotation.Z {
        type=MovieSceneScriptingFloatChannel
        default=0.1269
        keys [[2162,0.1269],[2180,0.1269],[2202,-2.7668],[2208,22.9233],[2211,51.3843],[2220,-19.1729],[2227,-6.1482],[2240,-2.3914],[2259,-19.2444],[2278,4.4671],[2296,-17.1971],[2306,-10.004],[2313,-15.0216],[2333,-5.188],[2351,-23.1494],[2374,-5.7727],[2383,-2.9225],[2414,-23.8956],[2432,-15.8716],[2441,-12.5426],[2454,-12.4983],[2477,-8.0047],[2493,-8.7396],[2515,-14.434],[2583,-6.4828],[2587,-6.0966],[2602,-5.9821],[2616,20.1427],[2634,12.1737],[2644,10.3635],[2656,6.5669],[2672,-4.3665],[2704,-6.0623],[2722,-9.1113],[2734,-9.1113],[2749,-52.697],[2754,-52.697],[2759,-0.9801]]
       }
       upperarm_l_fk_ctrl.Scale.X {type=MovieSceneScriptingFloatChannel, default=1, const [1,3]}
       upperarm_l_fk_ctrl.Scale.Y {type=MovieSceneScriptingFloatChannel, default=1, const [1,3]}
       upperarm_l_fk_ctrl.Scale.Z {type=MovieSceneScriptingFloatChannel, default=1, const [1,3]}
       lowerarm_l_fk_ctrl.Location.X {type=MovieSceneScriptingFloatChannel, default=0, const [0,6]}
       lowerarm_l_fk_ctrl.Location.Y {type=MovieSceneScriptingFloatChannel, default=0, const [0,6]}
       lowerarm_l_fk_ctrl.Location.Z {type=MovieSceneScriptingFloatChannel, default=0, const [0,6]}
       lowerarm_l_fk_ctrl.Rotation.X {type=MovieSceneScriptingFloatChannel, default=0, const [0,18]}
       lowerarm_l_fk_ctrl.Rotation.Y {type=MovieSceneScriptingFloatChannel, default=0, const [0,18]}
       lowerarm_l_fk_ctrl.Rotation.Z {
        type=MovieSceneScriptingFloatChannel
        default=0
        thinned_from=18
        keys [[2162,0],[2180,0],[2202,45.4376],[2208,11.6452],[2211,30.3303],[2220,-10.7902],[2586,-10.7015],[2612,6.827],[2644,6.827],[2655,6.8691],[2672,5.7064],[2704,6.0488],[2721,6.827],[2734,6.827],[2749,54.9523],[2754,54.9523],[2759,-15.1159]]
       }
       lowerarm_l_fk_ctrl.Scale.X {type=MovieSceneScriptingFloatChannel, default=1, const [1,3]}
       lowerarm_l_fk_ctrl.Scale.Y {type=MovieSceneScriptingFloatChannel, default=1, const [1,3]}
       lowerarm_l_fk_ctrl.Scale.Z {type=MovieSceneScriptingFloatChannel, default=1, const [1,3]}
       hand_l_fk_ctrl.Location.X {type=MovieSceneScriptingFloatChannel, default=0, const [0,6]}
       hand_l_fk_ctrl.Location.Y {type=MovieSceneScriptingFloatChannel, default=0, const [0,6]}
       hand_l_fk_ctrl.Location.Z {type=MovieSceneScriptingFloatChannel, default=0, const [0,6]}
       hand_l_fk_ctrl.Rotation.X {
        type=MovieSceneScriptingFloatChannel
        default=6.6278
        thinned_from=18
        keys [[2162,6.6278],[2180,6.6278],[2202,-58.3895],[2208,-28.9389],[2211,-28.9389],[2220,0.5124],[2586,0.5643],[2612,-17.5082],[2655,-17.5082],[2672,-2.2674],[2704,-6.9245],[2721,-17.5082],[2754,-17.5082],[2759,0]]
       }
       hand_l_fk_ctrl.Rotation.Y {
        type=MovieSceneScriptingFloatChannel
        default=-6.0098
        thinned_from=18
        keys [[2162,-6.0098],[2180,-6.0098],[2202,11.0965],[2208,10.6225],[2211,10.6225],[2220,10.1466],[2586,10.0108],[2612,-0.7098],[2655,-0.7098],[2672,2.1995],[2704,1.3105],[2721,-0.7098],[2754,-0.7098],[2759,0]]
       }
       hand_l_fk_ctrl.Rotation.Z {
        type=MovieSceneScriptingFloatChannel
        default=-1.6489
        thinned_from=18
        keys [[2162,-1.6489],[2180,-1.6489],[2202,-8.5917],[2208,-1.5127],[2211,-1.5127],[2220,5.5662],[2586,5.5068],[2612,-27.752],[2655,-27.752],[2672,-21.1656],[2704,-23.1782],[2721,-27.752],[2754,-27.752],[2759,0]]
       }
       hand_l_fk_ctrl.Scale.X {type=MovieSceneScriptingFloatChannel, default=1, const [1,3]}
       hand_l_fk_ctrl.Scale.Y {type=MovieSceneScriptingFloatChannel, default=1, const [1,3]}
       hand_l_fk_ctrl.Scale.Z {type=MovieSceneScriptingFloatChannel, default=1, const [1,3]}
       thumb_01_l_ctrl.Location.X {
        type=MovieSceneScriptingFloatChannel
        default=0.0001
        thinned_from=14
        keys [[2162,0.0001],[2586,0],[2645,0.0001],[2661,0],[2733,0.0001],[2754,0],[2759,0.0001]]
       }
       thumb_01_l_ctrl.Location.Y {type=MovieSceneScriptingFloatChannel, default=-0.0001, thinned_from=14, keys [[2162,-0.0001],[2759,0]]}
       thumb_01_l_ctrl.Location.Z {type=MovieSceneScriptingFloatChannel, default=-0.0001, thinned_from=14, keys [[2162,-0.0001],[2759,0]]}
       thumb_01_l_ctrl.Rotation.X {
        type=MovieSceneScriptingFloatChannel
        default=6.0844
        thinned_from=16
        keys [[2162,6.0844],[2202,-1.5583],[2586,-1.5583],[2612,6.0844],[2645,6.0844],[2661,-1.5583],[2665,-1.5544],[2676,4.4948],[2701,4.4948],[2708,3.7777],[2720,6.0844],[2733,6.0844],[2747,-6.5205],[2754,-6.5205],[2759,3.9725]]
       }
       thumb_01_l_ctrl.Rotation.Y {
        type=MovieSceneScriptingFloatChannel
        default=15.8669
        thinned_from=16
        keys [[2162,15.8669],[2202,0.8707],[2586,0.8707],[2612,15.8669],[2645,15.8669],[2661,0.8707],[2665,0.8783],[2676,12.7478],[2701,12.7478],[2708,11.3408],[2720,15.8669],[2733,15.8669],[2747,0],[2754,0],[2759,19.3454]]
       }
       thumb_01_l_ctrl.Rotation.Z {
        type=MovieSceneScriptingFloatChannel
        default=-4.4052
        thinned_from=16
        keys [[2162,-4.4052],[2202,-13.9227],[2586,-13.9227],[2612,-4.4052],[2645,-4.4052],[2661,-13.9227],[2665,-13.9179],[2676,-6.3847],[2701,-6.3847],[2708,-7.2777],[2720,-4.4052],[2733,-4.4052],[2747,0],[2754,0],[2759,-0.441]]
       }
       thumb_01_l_ctrl.Scale.X {type=MovieSceneScriptingFloatChannel, default=1, const [1,9]}
       thumb_01_l_ctrl.Scale.Y {type=MovieSceneScriptingFloatChannel, default=1, const [1,9]}
       thumb_01_l_ctrl.Scale.Z {type=MovieSceneScriptingFloatChannel, default=1, const [1,9]}
       thumb_02_l_ctrl.Location.X {
        type=MovieSceneScriptingFloatChannel
        default=-0.0001
        thinned_from=14
        keys [[2162,-0.0001],[2214,0],[2586,0],[2645,-0.0001],[2661,0],[2733,-0.0001],[2754,0],[2759,-0.0001]]
       }
       thumb_02_l_ctrl.Location.Y {type=MovieSceneScriptingFloatChannel, default=0, thinned_from=14, keys [[2162,0],[2759,0]]}
       thumb_02_l_ctrl.Location.Z {type=MovieSceneScriptingFloatChannel, default=0, thinned_from=14, keys [[2162,0],[2759,0]]}
       thumb_02_l_ctrl.Rotation.X {
        type=MovieSceneScriptingFloatChannel
        default=0.4753
        thinned_from=16
        keys [[2162,0.4753],[2202,0],[2586,0],[2612,0.4753],[2645,0.4753],[2661,0],[2665,0.0002],[2676,0.3765],[2701,0.3765],[2708,0.3319],[2720,0.4753],[2733,0.4753],[2747,-21.1821],[2754,-21.1821],[2759,0.123]]
       }
       thumb_02_l_ctrl.Rotation.Y {
        type=MovieSceneScriptingFloatChannel
        default=1.8999
        thinned_from=16
        keys [[2162,1.8999],[2202,0],[2586,0],[2612,1.8999],[2645,1.8999],[2661,0],[2665,0.001],[2676,1.5047],[2701,1.5047],[2708,1.3265],[2720,1.8999],[2733,1.8999],[2747,-0.3898],[2754,-0.3898],[2759,1.9494]]
       }
       thumb_02_l_ctrl.Rotation.Z {
        type=MovieSceneScriptingFloatChannel
        default=14.1339
        thinned_from=16
        keys [[2162,14.1339],[2202,0],[2586,0],[2612,14.1339],[2645,14.1339],[2661,0],[2665,0.0072],[2676,11.1942],[2701,11.1942],[2708,9.8681],[2720,14.1339],[2733,14.1339],[2747,19.2895],[2754,19.2895],[2759,3.6113]]
       }
       thumb_02_l_ctrl.Scale.X {type=MovieSceneScriptingFloatChannel, default=1, const [1,9]}
       thumb_02_l_ctrl.Scale.Y {type=MovieSceneScriptingFloatChannel, default=1, const [1,9]}
       thumb_02_l_ctrl.Scale.Z {type=MovieSceneScriptingFloatChannel, default=1, const [1,9]}
       thumb_03_l_ctrl.Location.X {type=MovieSceneScriptingFloatChannel, default=0, thinned_from=14, keys [[2162,0],[2759,0]]}
       thumb_03_l_ctrl.Location.Y {type=MovieSceneScriptingFloatChannel, default=0, thinned_from=14, keys [[2162,0],[2759,0]]}
       thumb_03_l_ctrl.Location.Z {type=MovieSceneScriptingFloatChannel, default=0, thinned_from=14, keys [[2162,0],[2759,0]]}
       thumb_03_l_ctrl.Rotation.X {
        type=MovieSceneScriptingFloatChannel
        default=1.3634
        thinned_from=16
        keys [[2162,1.3634],[2202,8.1645],[2586,8.1645],[2612,1.3634],[2645,1.3634],[2661,8.1645],[2665,8.161],[2676,2.778],[2701,2.778],[2708,3.4161],[2720,1.3634],[2733,1.3634],[2747,1.7533],[2754,1.7533],[2759,0.6645]]
       }
       thumb_03_l_ctrl.Rotation.Y {
        type=MovieSceneScriptingFloatChannel
        default=-1.4762
        thinned_from=16
        keys [[2162,-1.4762],[2202,-9.9357],[2586,-9.9357],[2612,-1.4762],[2645,-1.4762],[2661,-9.9357],[2665,-9.9314],[2676,-3.2357],[2701,-3.2357],[2708,-4.0294],[2720,-1.4762],[2733,-1.4762],[2747,-0.9817],[2754,-0.9817],[2759,-1.8964]]
       }
       thumb_03_l_ctrl.Rotation.Z {
        type=MovieSceneScriptingFloatChannel
        default=0
        thinned_from=16
        keys [[2162,0],[2202,-50.7089],[2586,-50.7089],[2612,0],[2645,0],[2661,-50.7089],[2665,-50.6831],[2676,-10.547],[2701,-10.547],[2708,-15.3048],[2720,0],[2733,0],[2747,0.0764],[2754,0.0764],[2759,23.4201]]
       }
       thumb_03_l_ctrl.Scale.X {type=MovieSceneScriptingFloatChannel, default=1, const [1,9]}
       thumb_03_l_ctrl.Scale.Y {type=MovieSceneScriptingFloatChannel, default=1, const [1,9]}
       thumb_03_l_ctrl.Scale.Z {type=MovieSceneScriptingFloatChannel, default=1, const [1,9]}
       index_metacarpal_l_ctrl.Location.X {
        type=MovieSceneScriptingFloatChannel
        default=-0.0001
        thinned_from=14
        keys [[2162,-0.0001],[2214,0],[2586,0],[2645,-0.0001],[2661,0],[2733,-0.0001],[2754,0],[2759,-0.0001]]
       }
       index_metacarpal_l_ctrl.Location.Y {
        type=MovieSceneScriptingFloatChannel
        default=-0.0002
        thinned_from=14
        keys [[2162,-0.0002],[2214,0],[2586,0],[2612,-0.0002],[2645,-0.0002],[2661,0],[2676,-0.0002],[2733,-0.0002],[2754,0],[2759,-0.0002]]
       }
       index_metacarpal_l_ctrl.Location.Z {type=MovieSceneScriptingFloatChannel, default=0, thinned_from=14, keys [[2162,0],[2759,0]]}
       index_metacarpal_l_ctrl.Rotation.X {type=MovieSceneScriptingFloatChannel, default=0, const [0,16]}
       index_metacarpal_l_ctrl.Rotation.Y {
        type=MovieSceneScriptingFloatChannel
        default=0
        thinned_from=16
        keys [[2162,0],[2202,3.7815],[2586,3.7815],[2612,0],[2645,0],[2661,3.7815],[2665,3.7796],[2676,0.7865],[2701,0.7865],[2708,1.1413],[2720,0],[2733,0],[2747,1.7196],[2759,1.7196]]
       }
       index_metacarpal_l_ctrl.Rotation.Z {
        type=MovieSceneScriptingFloatChannel
        default=0
        thinned_from=16
        keys [[2162,0],[2202,-3.094],[2586,-3.094],[2612,0],[2645,0],[2661,-3.094],[2665,-3.0924],[2676,-0.6435],[2701,-0.6435],[2708,-0.9338],[2720,0],[2759,0]]
       }
       index_metacarpal_l_ctrl.Scale.X {type=MovieSceneScriptingFloatChannel, default=1, const [1,9]}
       index_metacarpal_l_ctrl.Scale.Y {type=MovieSceneScriptingFloatChannel, default=1, const [1,9]}
       index_metacarpal_l_ctrl.Scale.Z {type=MovieSceneScriptingFloatChannel, default=1, const [1,9]}
       index_01_l_ctrl.Location.X {type=MovieSceneScriptingFloatChannel, default=0, thinned_from=14, keys [[2162,0],[2759,0]]}
       index_01_l_ctrl.Location.Y {type=MovieSceneScriptingFloatChannel, default=0, thinned_from=14, keys [[2162,0],[2759,0]]}
       index_01_l_ctrl.Location.Z {type=MovieSceneScriptingFloatChannel, default=0, thinned_from=14, keys [[2162,0],[2759,0]]}
       index_01_l_ctrl.Rotation.X {
        type=MovieSceneScriptingFloatChannel
        default=0
        thinned_from=16
        keys [[2162,0],[2202,6.2821],[2586,6.2821],[2612,0],[2645,0],[2661,6.2821],[2665,6.2789],[2676,1.3066],[2701,1.3066],[2708,1.896],[2720,0],[2733,0],[2747,-2.7924],[2754,-2.7924],[2759,-4.7213]]
       }
       index_01_l_ctrl.Rotation.Y {
        type=MovieSceneScriptingFloatChannel
        default=0
        thinned_from=16
        keys [[2162,0],[2202,-2.5276],[2586,-2.5276],[2612,0],[2645,0],[2661,-2.5276],[2665,-2.5264],[2676,-0.5257],[2701,-0.5257],[2708,-0.7629],[2720,0],[2733,0],[2747,-0.974],[2754,-0.974],[2759,7.9585]]
       }
       index_01_l_ctrl.Rotation.Z {
        type=MovieSceneScriptingFloatChannel
        default=7.2214
        thinned_from=16
        keys [[2162,7.2214],[2202,-65.1155],[2586,-65.1155],[2612,7.2214],[2645,7.2214],[2661,-65.1155],[2665,-65.0787],[2676,-7.824],[2701,-7.824],[2708,-14.611],[2720,7.2214],[2733,7.2214],[2747,8.9721],[2754,8.9721],[2759,31.1517]]
       }
       index_01_l_ctrl.Scale.X {type=MovieSceneScriptingFloatChannel, default=1, const [1,9]}
       index_01_l_ctrl.Scale.Y {type=MovieSceneScriptingFloatChannel, default=1, const [1,9]}
       index_01_l_ctrl.Scale.Z {type=MovieSceneScriptingFloatChannel, default=1, const [1,9]}
       index_02_l_ctrl.Location.X {type=MovieSceneScriptingFloatChannel, default=0.0001, thinned_from=14, keys [[2162,0.0001],[2759,0]]}
       index_02_l_ctrl.Location.Y {
        type=MovieSceneScriptingFloatChannel
        default=0.0001
        thinned_from=14
        keys [[2162,0.0001],[2586,0],[2645,0.0001],[2661,0],[2733,0.0001],[2754,0],[2759,0.0001]]
       }
       index_02_l_ctrl.Location.Z {type=MovieSceneScriptingFloatChannel, default=0.0001, thinned_from=14, keys [[2162,0.0001],[2759,0]]}
       index_02_l_ctrl.Rotation.X {
        type=MovieSceneScriptingFloatChannel
        default=-0.23
        thinned_from=16
        keys [[2162,-0.23],[2202,0],[2586,0],[2612,-0.23],[2645,-0.23],[2661,0],[2665,-0.0001],[2676,-0.1822],[2701,-0.1822],[2708,-0.1606],[2720,-0.23],[2733,-0.23],[2747,-0.3304],[2754,-0.3304],[2759,-0.1354]]
       }
       index_02_l_ctrl.Rotation.Y {
        type=MovieSceneScriptingFloatChannel
        default=0.3153
        thinned_from=16
        keys [[2162,0.3153],[2202,0],[2586,0],[2612,0.3153],[2645,0.3153],[2661,0],[2665,0.0002],[2676,0.2497],[2701,0.2497],[2708,0.2201],[2720,0.3153],[2733,0.3153],[2747,0.2098],[2754,0.2098],[2759,3.7485]]
       }
       index_02_l_ctrl.Rotation.Z {
        type=MovieSceneScriptingFloatChannel
        default=20.8963
        thinned_from=16
        keys [[2162,20.8963],[2202,-60.8298],[2586,-60.8298],[2612,20.8963],[2645,20.8963],[2661,-60.8298],[2665,-60.7882],[2676,3.898],[2701,3.898],[2708,-3.77],[2720,20.8963],[2733,20.8963],[2747,13.9361],[2754,13.9361],[2759,35.3799]]
       }
       index_02_l_ctrl.Scale.X {type=MovieSceneScriptingFloatChannel, default=1, const [1,9]}
       index_02_l_ctrl.Scale.Y {type=MovieSceneScriptingFloatChannel, default=1, const [1,9]}
       index_02_l_ctrl.Scale.Z {type=MovieSceneScriptingFloatChannel, default=1, const [1,9]}
       index_03_l_ctrl.Location.X {type=MovieSceneScriptingFloatChannel, default=0.0001, thinned_from=14, keys [[2162,0.0001],[2759,0]]}
       index_03_l_ctrl.Location.Y {type=MovieSceneScriptingFloatChannel, default=0, thinned_from=14, keys [[2162,0],[2759,0]]}
       index_03_l_ctrl.Location.Z {type=MovieSceneScriptingFloatChannel, default=0.0001, thinned_from=14, keys [[2162,0.0001],[2759,0]]}
       index_03_l_ctrl.Rotation.X {
        type=MovieSceneScriptingFloatChannel
        default=-0.0867
        thinned_from=16
        keys [[2162,-0.0867],[2202,0.0236],[2586,0.0236],[2612,-0.0867],[2645,-0.0867],[2661,0.0236],[2665,0.0236],[2676,-0.0637],[2701,-0.0637],[2708,-0.0534],[2720,-0.0867],[2733,-0.0867],[2747,-0.5169],[2754,-0.5169],[2759,-0.0057]]
       }
       index_03_l_ctrl.Rotation.Y {
        type=MovieSceneScriptingFloatChannel
        default=0.7964
        thinned_from=16
        keys [[2162,0.7964],[2202,-0.109],[2586,-0.109],[2612,0.7964],[2645,0.7964],[2661,-0.109],[2665,-0.1086],[2676,0.6081],[2701,0.6081],[2708,0.5232],[2720,0.7964],[2733,0.7964],[2747,0.6557],[2754,0.6557],[2759,0.8011]]
       }
       index_03_l_ctrl.Rotation.Z {
        type=MovieSceneScriptingFloatChannel
        default=25.1985
        thinned_from=16
        keys [[2162,25.1985],[2202,-78.806],[2586,-78.806],[2612,25.1985],[2645,25.1985],[2661,-78.806],[2665,-78.7531],[2676,3.5665],[2701,3.5665],[2708,-6.1917],[2720,25.1985],[2733,25.1985],[2747,6.4585],[2754,6.4585],[2759,31.0033]]
       }
       index_03_l_ctrl.Scale.X {type=MovieSceneScriptingFloatChannel, default=1, const [1,9]}
       index_03_l_ctrl.Scale.Y {type=MovieSceneScriptingFloatChannel, default=1, const [1,9]}
       index_03_l_ctrl.Scale.Z {type=MovieSceneScriptingFloatChannel, default=1, const [1,9]}
       middle_metacarpal_l_ctrl.Location.X {type=MovieSceneScriptingFloatChannel, default=0, thinned_from=14, keys [[2162,0],[2759,0]]}
       middle_metacarpal_l_ctrl.Location.Y {
        type=MovieSceneScriptingFloatChannel
        default=-0.0001
        thinned_from=14
        keys [[2162,-0.0001],[2214,0],[2586,0],[2645,-0.0001],[2661,0],[2733,-0.0001],[2754,0],[2759,-0.0001]]
       }
       middle_metacarpal_l_ctrl.Location.Z {type=MovieSceneScriptingFloatChannel, default=0, thinned_from=14, keys [[2162,0],[2759,0]]}
       middle_metacarpal_l_ctrl.Rotation.X {
        type=MovieSceneScriptingFloatChannel
        default=0
        thinned_from=16
        keys [[2162,0],[2202,-7.29],[2586,-7.29],[2612,0],[2645,0],[2661,-7.29],[2665,-7.2863],[2676,-1.5163],[2701,-1.5163],[2708,-2.2002],[2720,0],[2759,0]]
       }
       middle_metacarpal_l_ctrl.Rotation.Y {
        type=MovieSceneScriptingFloatChannel
        default=0
        thinned_from=16
        keys [[2162,0],[2202,1.3582],[2586,1.3582],[2612,0],[2645,0],[2661,1.3582],[2665,1.3575],[2676,0.2825],[2701,0.2825],[2708,0.4099],[2720,0],[2733,0],[2747,4.0107],[2754,4.0107],[2759,0]]
       }
       middle_metacarpal_l_ctrl.Rotation.Z {
        type=MovieSceneScriptingFloatChannel
        default=0
        thinned_from=16
        keys [[2162,0],[2202,-2.7048],[2586,-2.7048],[2612,0],[2645,0],[2661,-2.7048],[2665,-2.7034],[2676,-0.5626],[2701,-0.5626],[2708,-0.8164],[2720,0],[2759,0]]
       }
       middle_metacarpal_l_ctrl.Scale.X {type=MovieSceneScriptingFloatChannel, default=1, const [1,9]}
       middle_metacarpal_l_ctrl.Scale.Y {type=MovieSceneScriptingFloatChannel, default=1, const [1,9]}
       middle_metacarpal_l_ctrl.Scale.Z {type=MovieSceneScriptingFloatChannel, default=1, const [1,9]}
       middle_01_l_ctrl.Location.X {
        type=MovieSceneScriptingFloatChannel
        default=-0.0002
        thinned_from=14
        keys [[2162,-0.0002],[2214,0],[2586,0],[2612,-0.0002],[2645,-0.0002],[2661,0],[2676,-0.0002],[2733,-0.0002],[2754,0],[2759,-0.0002]]
       }
       middle_01_l_ctrl.Location.Y {type=MovieSceneScriptingFloatChannel, default=0, thinned_from=14, keys [[2162,0],[2759,0]]}
       middle_01_l_ctrl.Location.Z {type=MovieSceneScriptingFloatChannel, default=0, thinned_from=14, keys [[2162,0],[2759,0]]}
       middle_01_l_ctrl.Rotation.X {type=MovieSceneScriptingFloatChannel, default=0, thinned_from=16, keys [[2162,0],[2733,0],[2747,-6.8374],[2754,-6.8374],[2759,-3.9745]]}
       middle_01_l_ctrl.Rotation.Y {
        type=MovieSceneScriptingFloatChannel
        default=0
        thinned_from=16
        keys [[2162,0],[2202,1.3494],[2586,1.3494],[2612,0],[2645,0],[2661,1.3494],[2665,1.3487],[2676,0.2807],[2701,0.2807],[2708,0.4073],[2720,0],[2733,0],[2747,-5.3889],[2754,-5.3889],[2759,0]]
       }
       middle_01_l_ctrl.Rotation.Z {
        type=MovieSceneScriptingFloatChannel
        default=3.8267
        thinned_from=16
        keys [[2162,3.8267],[2202,-74.7137],[2586,-74.7137],[2612,3.8267],[2645,3.8267],[2661,-74.7137],[2665,-74.6738],[2676,-12.509],[2701,-12.509],[2708,-19.8781],[2720,3.8267],[2733,3.8267],[2747,-11.0412],[2754,-11.0412],[2759,18.3246]]
       }
       middle_01_l_ctrl.Scale.X {type=MovieSceneScriptingFloatChannel, default=1, const [1,9]}
       middle_01_l_ctrl.Scale.Y {type=MovieSceneScriptingFloatChannel, default=1, const [1,9]}
       middle_01_l_ctrl.Scale.Z {type=MovieSceneScriptingFloatChannel, default=1, const [1,9]}
       middle_02_l_ctrl.Location.X {
        type=MovieSceneScriptingFloatChannel
        default=-0.0008
        thinned_from=14
        keys [[2162,-0.0008],[2202,-0.0452],[2586,-0.0452],[2612,-0.0008],[2645,-0.0008],[2661,-0.0452],[2676,-0.0068],[2701,-0.0068],[2720,-0.0008],[2733,-0.0008],[2747,-0.0452],[2754,-0.0452],[2759,-0.0008]]
       }
       middle_02_l_ctrl.Location.Y {
        type=MovieSceneScriptingFloatChannel
        default=-0.0009
        thinned_from=14
        keys [[2162,-0.0009],[2202,0.0507],[2586,0.0507],[2612,-0.0009],[2645,-0.0009],[2661,0.0507],[2676,0.0062],[2701,0.0062],[2720,-0.0008],[2733,-0.0008],[2747,0.0507],[2754,0.0507],[2759,-0.0009]]
       }
       middle_02_l_ctrl.Location.Z {
        type=MovieSceneScriptingFloatChannel
        default=-0.0476
        thinned_from=14
        keys [[2162,-0.0476],[2202,0.1146],[2586,0.1146],[2612,-0.0476],[2645,-0.0476],[2661,0.1146],[2676,-0.0255],[2701,-0.0255],[2720,-0.0476],[2733,-0.0476],[2747,0.1146],[2754,0.1146],[2759,-0.0476]]
       }
       middle_02_l_ctrl.Rotation.X {
        type=MovieSceneScriptingFloatChannel
        default=-2.9188
        thinned_from=16
        keys [[2162,-2.9188],[2202,-3.3929],[2586,-3.3929],[2612,-2.9188],[2645,-2.9188],[2661,-3.3929],[2665,-3.3926],[2676,-3.0174],[2701,-3.0174],[2708,-3.0619],[2720,-2.9188],[2733,-2.9188],[2747,-6.9475],[2754,-6.9475],[2759,-1.4489]]
       }
       middle_02_l_ctrl.Rotation.Y {
        type=MovieSceneScriptingFloatChannel
        default=2.5438
        thinned_from=16
        keys [[2162,2.5438],[2202,-1.8646],[2586,-1.8646],[2612,2.5438],[2645,2.5438],[2661,-1.8646],[2665,-1.8623],[2676,1.6269],[2701,1.6269],[2708,1.2133],[2720,2.5438],[2733,2.5438],[2747,3.3291],[2754,3.3291],[2759,6.9733]]
       }
       middle_02_l_ctrl.Rotation.Z {
        type=MovieSceneScriptingFloatChannel
        default=12.0645
        thinned_from=16
        keys [[2162,12.0645],[2202,-58.0319],[2586,-58.0319],[2612,12.0645],[2645,12.0645],[2661,-58.0319],[2665,-57.9963],[2676,-2.5149],[2701,-2.5149],[2708,-9.0917],[2720,12.0645],[2733,12.0645],[2747,2.1506],[2754,2.1506],[2759,39.0682]]
       }
       middle_02_l_ctrl.Scale.X {type=MovieSceneScriptingFloatChannel, default=1, const [1,9]}
       middle_02_l_ctrl.Scale.Y {type=MovieSceneScriptingFloatChannel, default=1, const [1,9]}
       middle_02_l_ctrl.Scale.Z {type=MovieSceneScriptingFloatChannel, default=1, const [1,9]}
       middle_03_l_ctrl.Location.X {
        type=MovieSceneScriptingFloatChannel
        default=-0.0005
        thinned_from=14
        keys [[2162,-0.0005],[2202,0.0206],[2586,0.0206],[2612,-0.0005],[2645,-0.0005],[2661,0.0206],[2676,0.0024],[2701,0.0024],[2720,-0.0005],[2733,-0.0005],[2747,0.0206],[2754,0.0206],[2759,-0.0005]]
       }
       middle_03_l_ctrl.Location.Y {
        type=MovieSceneScriptingFloatChannel
        default=0
        thinned_from=14
        keys [[2162,0],[2202,-0.0583],[2586,-0.0583],[2612,0],[2645,0],[2661,-0.0583],[2676,-0.008],[2701,-0.008],[2720,0],[2733,0],[2747,-0.0583],[2754,-0.0583],[2759,0]]
       }
       middle_03_l_ctrl.Location.Z {
        type=MovieSceneScriptingFloatChannel
        default=-0.0061
        thinned_from=14
        keys [[2162,-0.0061],[2202,0.1324],[2586,0.1324],[2612,-0.0061],[2645,-0.0061],[2661,0.1324],[2676,0.0127],[2701,0.0127],[2720,-0.0061],[2733,-0.0061],[2747,0.1324],[2754,0.1324],[2759,-0.0061]]
       }
       middle_03_l_ctrl.Rotation.X {
        type=MovieSceneScriptingFloatChannel
        default=1.0874
        thinned_from=16
        keys [[2162,1.0874],[2202,-1.8232],[2586,-1.8232],[2612,1.0874],[2645,1.0874],[2661,-1.8232],[2665,-1.8218],[2676,0.482],[2701,0.482],[2708,0.2089],[2720,1.0874],[2733,1.0874],[2747,2.0265],[2754,2.0265],[2759,1.6933]]
       }
       middle_03_l_ctrl.Rotation.Y {
        type=MovieSceneScriptingFloatChannel
        default=2.0851
        thinned_from=16
        keys [[2162,2.0851],[2202,1.4852],[2586,1.4852],[2612,2.0851],[2645,2.0851],[2661,1.4852],[2665,1.4856],[2676,1.9604],[2701,1.9604],[2708,1.9041],[2720,2.0851],[2733,2.0851],[2747,4.2407],[2754,4.2407],[2759,1.6319]]
       }
       middle_03_l_ctrl.Rotation.Z {
        type=MovieSceneScriptingFloatChannel
        default=14.0525
        thinned_from=16
        keys [[2162,14.0525],[2202,-64.6207],[2586,-64.6207],[2612,14.0525],[2645,14.0525],[2661,-64.6207],[2665,-64.5807],[2676,-2.3108],[2701,-2.3108],[2708,-9.6923],[2720,14.0525],[2733,14.0525],[2754,14.0851],[2759,32.5715]]
       }
       middle_03_l_ctrl.Scale.X {type=MovieSceneScriptingFloatChannel, default=1, const [1,9]}
       middle_03_l_ctrl.Scale.Y {type=MovieSceneScriptingFloatChannel, default=1, const [1,9]}
       middle_03_l_ctrl.Scale.Z {type=MovieSceneScriptingFloatChannel, default=1, const [1,9]}
       ring_metacarpal_l_ctrl.Location.X {
        type=MovieSceneScriptingFloatChannel
        default=-0.0001
        thinned_from=14
        keys [[2162,-0.0001],[2214,0],[2586,0],[2645,-0.0001],[2661,0],[2733,-0.0001],[2754,0],[2759,-0.0001]]
       }
       ring_metacarpal_l_ctrl.Location.Y {
        type=MovieSceneScriptingFloatChannel
        default=-0.0002
        thinned_from=14
        keys [[2162,-0.0002],[2214,0],[2586,0],[2645,-0.0002],[2661,0],[2733,-0.0002],[2754,0],[2759,-0.0002]]
       }
       ring_metacarpal_l_ctrl.Location.Z {type=MovieSceneScriptingFloatChannel, default=0, thinned_from=14, keys [[2162,0],[2759,0]]}
       ring_metacarpal_l_ctrl.Rotation.X {
        type=MovieSceneScriptingFloatChannel
        default=0
        thinned_from=16
        keys [[2162,0],[2202,-2.4847],[2586,-2.4847],[2612,0],[2645,0],[2661,-2.4847],[2665,-2.4835],[2676,-0.5168],[2701,-0.5168],[2708,-0.7499],[2720,0],[2759,0]]
       }
       ring_metacarpal_l_ctrl.Rotation.Y {
        type=MovieSceneScriptingFloatChannel
        default=0
        thinned_from=16
        keys [[2162,0],[2202,-0.8637],[2586,-0.8637],[2612,0],[2645,0],[2661,-0.8637],[2665,-0.8632],[2676,-0.1796],[2701,-0.1796],[2708,-0.2607],[2720,0],[2733,0],[2747,1.4724],[2754,1.4724],[2759,-1.1632]]
       }
       ring_metacarpal_l_ctrl.Rotation.Z {
        type=MovieSceneScriptingFloatChannel
        default=0
        thinned_from=16
        keys [[2162,0],[2202,-5.6698],[2586,-5.6698],[2612,0],[2645,0],[2661,-5.6698],[2665,-5.6669],[2676,-1.1793],[2701,-1.1793],[2708,-1.7112],[2720,0],[2759,0]]
       }
       ring_metacarpal_l_ctrl.Scale.X {type=MovieSceneScriptingFloatChannel, default=1, const [1,9]}
       ring_metacarpal_l_ctrl.Scale.Y {type=MovieSceneScriptingFloatChannel, default=1, const [1,9]}
       ring_metacarpal_l_ctrl.Scale.Z {type=MovieSceneScriptingFloatChannel, default=1, const [1,9]}
       ring_01_l_ctrl.Location.X {
        type=MovieSceneScriptingFloatChannel
        default=0.0002
        thinned_from=14
        keys [[2162,0.0002],[2214,0],[2586,0],[2612,0.0002],[2645,0.0002],[2661,0],[2733,0.0002],[2754,0],[2759,0.0002]]
       }
       ring_01_l_ctrl.Location.Y {type=MovieSceneScriptingFloatChannel, default=0, thinned_from=14, keys [[2162,0],[2759,0]]}
       ring_01_l_ctrl.Location.Z {type=MovieSceneScriptingFloatChannel, default=0, thinned_from=14, keys [[2162,0],[2759,0]]}
       ring_01_l_ctrl.Rotation.X {
        type=MovieSceneScriptingFloatChannel
        default=0
        thinned_from=16
        keys [[2162,0],[2202,-4.509],[2586,-4.509],[2612,0],[2645,0],[2661,-4.509],[2665,-4.5067],[2676,-0.9378],[2701,-0.9378],[2708,-1.3609],[2720,0],[2733,0],[2747,-10.1829],[2754,-10.1829],[2759,10.5727]]
       }
       ring_01_l_ctrl.Rotation.Y {
        type=MovieSceneScriptingFloatChannel
        default=-0.3119
        thinned_from=16
        keys [[2162,-0.3119],[2202,-1.201],[2586,-1.201],[2612,-0.3119],[2645,-0.3119],[2661,-1.201],[2665,-1.2005],[2676,-0.4968],[2701,-0.4968],[2708,-0.5802],[2720,-0.3119],[2733,-0.3119],[2747,-4.284],[2754,-4.284],[2759,-4.6626]]
       }
       ring_01_l_ctrl.Rotation.Z {
        type=MovieSceneScriptingFloatChannel
        default=0
        thinned_from=16
        keys [[2162,0],[2202,-79.0231],[2586,-79.0231],[2612,0],[2645,0],[2661,-79.0231],[2665,-78.9829],[2676,-16.4361],[2701,-16.4361],[2708,-23.8504],[2720,0],[2733,0],[2747,-34.3986],[2754,-34.3986],[2759,22.2578]]
       }
       ring_01_l_ctrl.Scale.X {type=MovieSceneScriptingFloatChannel, default=1, const [1,9]}
       ring_01_l_ctrl.Scale.Y {type=MovieSceneScriptingFloatChannel, default=1, const [1,9]}
       ring_01_l_ctrl.Scale.Z {type=MovieSceneScriptingFloatChannel, default=1, const [1,9]}
       ring_02_l_ctrl.Location.X {
        type=MovieSceneScriptingFloatChannel
        default=-0.0003
        thinned_from=14
        keys [[2162,-0.0003],[2214,0],[2586,0],[2612,-0.0003],[2645,-0.0003],[2661,0],[2676,-0.0002],[2733,-0.0003],[2754,0],[2759,-0.0003]]
       }
       ring_02_l_ctrl.Location.Y {type=MovieSceneScriptingFloatChannel, default=-0.0001, thinned_from=14, keys [[2162,-0.0001],[2759,0]]}
       ring_02_l_ctrl.Location.Z {type=MovieSceneScriptingFloatChannel, default=0, thinned_from=14, keys [[2162,0],[2759,0]]}
       ring_02_l_ctrl.Rotation.X {type=MovieSceneScriptingFloatChannel, default=0, const [0,16]}
       ring_02_l_ctrl.Rotation.Y {type=MovieSceneScriptingFloatChannel, default=0, thinned_from=16, keys [[2162,0],[2754,0],[2759,3.3812]]}
       ring_02_l_ctrl.Rotation.Z {
        type=MovieSceneScriptingFloatChannel
        default=0
        thinned_from=16
        keys [[2162,0],[2202,-66.6687],[2586,-66.6687],[2612,0],[2645,0],[2661,-66.6687],[2665,-66.6348],[2676,-13.8665],[2701,-13.8665],[2708,-20.1217],[2720,0],[2733,0],[2747,-7.8478],[2754,-7.8478],[2759,35.2878]]
       }
       ring_02_l_ctrl.Scale.X {type=MovieSceneScriptingFloatChannel, default=1, const [1,9]}
       ring_02_l_ctrl.Scale.Y {type=MovieSceneScriptingFloatChannel, default=1, const [1,9]}
       ring_02_l_ctrl.Scale.Z {type=MovieSceneScriptingFloatChannel, default=1, const [1,9]}
       ring_03_l_ctrl.Location.X {type=MovieSceneScriptingFloatChannel, default=0, const [0,14]}
       ring_03_l_ctrl.Location.Y {type=MovieSceneScriptingFloatChannel, default=0, const [0,14]}
       ring_03_l_ctrl.Location.Z {type=MovieSceneScriptingFloatChannel, default=0, const [0,14]}
       ring_03_l_ctrl.Rotation.X {type=MovieSceneScriptingFloatChannel, default=0, const [0,16]}
       ring_03_l_ctrl.Rotation.Y {type=MovieSceneScriptingFloatChannel, default=0, const [0,16]}
       ring_03_l_ctrl.Rotation.Z {
        type=MovieSceneScriptingFloatChannel
        default=0
        thinned_from=16
        keys [[2162,0],[2202,-73.2004],[2586,-73.2004],[2612,0],[2645,0],[2661,-73.2004],[2665,-73.1632],[2676,-15.225],[2701,-15.225],[2708,-22.0931],[2720,0],[2733,0],[2747,-0.6769],[2754,-0.6769],[2759,20.858]]
       }
       ring_03_l_ctrl.Scale.X {type=MovieSceneScriptingFloatChannel, default=1, const [1,9]}
       ring_03_l_ctrl.Scale.Y {type=MovieSceneScriptingFloatChannel, default=1, const [1,9]}
       ring_03_l_ctrl.Scale.Z {type=MovieSceneScriptingFloatChannel, default=1, const [1,9]}
       pinky_metacarpal_l_ctrl.Location.X {type=MovieSceneScriptingFloatChannel, default=0, const [0,14]}
       pinky_metacarpal_l_ctrl.Location.Y {type=MovieSceneScriptingFloatChannel, default=0, const [0,14]}
       pinky_metacarpal_l_ctrl.Location.Z {type=MovieSceneScriptingFloatChannel, default=0, const [0,14]}
       pinky_metacarpal_l_ctrl.Rotation.X {
        type=MovieSceneScriptingFloatChannel
        default=0
        thinned_from=16
        keys [[2162,0],[2202,-5.9588],[2586,-5.9588],[2612,0],[2645,0],[2661,-5.9588],[2665,-5.9558],[2676,-1.2394],[2701,-1.2394],[2708,-1.7985],[2720,0],[2759,0]]
       }
       pinky_metacarpal_l_ctrl.Rotation.Y {
        type=MovieSceneScriptingFloatChannel
        default=-3.1934
        thinned_from=16
        keys [[2162,-3.1934],[2202,-5.2205],[2586,-5.2205],[2612,-3.1934],[2645,-3.1934],[2661,-5.2205],[2665,-5.2195],[2676,-3.615],[2701,-3.615],[2708,-3.8052],[2720,-3.1934],[2759,-3.1934]]
       }
       pinky_metacarpal_l_ctrl.Rotation.Z {
        type=MovieSceneScriptingFloatChannel
        default=0
        thinned_from=16
        keys [[2162,0],[2202,-11.0008],[2586,-11.0008],[2612,0],[2645,0],[2661,-11.0008],[2665,-10.9952],[2676,-2.2881],[2701,-2.2881],[2708,-3.3202],[2720,0],[2759,0]]
       }
       pinky_metacarpal_l_ctrl.Scale.X {type=MovieSceneScriptingFloatChannel, default=1, const [1,9]}
       pinky_metacarpal_l_ctrl.Scale.Y {type=MovieSceneScriptingFloatChannel, default=1, const [1,9]}
       pinky_metacarpal_l_ctrl.Scale.Z {type=MovieSceneScriptingFloatChannel, default=1, const [1,9]}
       pinky_01_l_ctrl.Location.X {
        type=MovieSceneScriptingFloatChannel
        default=-0.0001
        thinned_from=14
        keys [[2162,-0.0001],[2214,0],[2586,0],[2645,-0.0001],[2661,0],[2733,-0.0001],[2754,0],[2759,-0.0001]]
       }
       pinky_01_l_ctrl.Location.Y {
        type=MovieSceneScriptingFloatChannel
        default=-0.0002
        thinned_from=14
        keys [[2162,-0.0002],[2214,0],[2586,0],[2612,-0.0002],[2645,-0.0002],[2661,0],[2676,-0.0002],[2733,-0.0002],[2754,0],[2759,-0.0002]]
       }
       pinky_01_l_ctrl.Location.Z {type=MovieSceneScriptingFloatChannel, default=0, thinned_from=14, keys [[2162,0],[2759,0]]}
       pinky_01_l_ctrl.Rotation.X {
        type=MovieSceneScriptingFloatChannel
        default=-4.1261
        thinned_from=16
        keys [[2162,-4.1261],[2202,-11.7646],[2586,-11.7646],[2612,-4.1261],[2645,-4.1261],[2661,-11.7646],[2665,-11.7607],[2676,-5.7149],[2701,-5.7149],[2708,-6.4315],[2720,-4.1261],[2733,-4.1261],[2747,-18.6348],[2754,-18.6348],[2759,14.6395]]
       }
       pinky_01_l_ctrl.Rotation.Y {
        type=MovieSceneScriptingFloatChannel
        default=1.0722
        thinned_from=16
        keys [[2162,1.0722],[2202,-1.0914],[2586,-1.0914],[2612,1.0722],[2645,1.0722],[2661,-1.0914],[2665,-1.0903],[2676,0.6222],[2701,0.6222],[2708,0.4192],[2720,1.0722],[2733,1.0722],[2747,-1.3004],[2754,-1.3004],[2759,-11.264]]
       }
       pinky_01_l_ctrl.Rotation.Z {
        type=MovieSceneScriptingFloatChannel
        default=-14.4594
        thinned_from=16
        keys [[2162,-14.4594],[2202,-84.6646],[2586,-84.6646],[2612,-14.4594],[2645,-14.4594],[2661,-84.6646],[2665,-84.6289],[2676,-29.0615],[2701,-29.0615],[2708,-35.6485],[2720,-14.4594],[2733,-14.4594],[2747,-45.6936],[2754,-45.6936],[2759,24.5562]]
       }
       pinky_01_l_ctrl.Scale.X {type=MovieSceneScriptingFloatChannel, default=1, const [1,9]}
       pinky_01_l_ctrl.Scale.Y {type=MovieSceneScriptingFloatChannel, default=1, const [1,9]}
       pinky_01_l_ctrl.Scale.Z {type=MovieSceneScriptingFloatChannel, default=1, const [1,9]}
       pinky_02_l_ctrl.Location.X {type=MovieSceneScriptingFloatChannel, default=0.0001, thinned_from=14, keys [[2162,0.0001],[2759,0]]}
       pinky_02_l_ctrl.Location.Y {type=MovieSceneScriptingFloatChannel, default=-0.0001, thinned_from=14, keys [[2162,-0.0001],[2759,0]]}
       pinky_02_l_ctrl.Location.Z {type=MovieSceneScriptingFloatChannel, default=0, thinned_from=14, keys [[2162,0],[2759,0]]}
       pinky_02_l_ctrl.Rotation.X {
        type=MovieSceneScriptingFloatChannel
        default=0.3719
        thinned_from=16
        keys [[2162,0.3719],[2202,0],[2586,0],[2612,0.3719],[2645,0.3719],[2661,0],[2665,0.0002],[2676,0.2946],[2701,0.2946],[2708,0.2597],[2720,0.3719],[2733,0.3719],[2747,0.267],[2754,0.267],[2759,0.4415]]
       }
       pinky_02_l_ctrl.Rotation.Y {
        type=MovieSceneScriptingFloatChannel
        default=0.24
        thinned_from=16
        keys [[2162,0.24],[2202,0],[2586,0],[2612,0.24],[2645,0.24],[2661,0],[2665,0.0001],[2676,0.1901],[2701,0.1901],[2708,0.1676],[2720,0.24],[2733,0.24],[2747,0.3549],[2754,0.3549],[2759,3.4359]]
       }
       pinky_02_l_ctrl.Rotation.Z {
        type=MovieSceneScriptingFloatChannel
        default=2.0416
        thinned_from=16
        keys [[2162,2.0416],[2202,-63.0936],[2586,-63.0936],[2612,2.0416],[2645,2.0416],[2661,-63.0936],[2665,-63.0605],[2676,-11.506],[2701,-11.506],[2708,-17.6173],[2720,2.0416],[2733,2.0416],[2747,-15.8625],[2754,-15.8625],[2759,26.1087]]
       }
       pinky_02_l_ctrl.Scale.X {type=MovieSceneScriptingFloatChannel, default=1, const [1,9]}
       pinky_02_l_ctrl.Scale.Y {type=MovieSceneScriptingFloatChannel, default=1, const [1,9]}
       pinky_02_l_ctrl.Scale.Z {type=MovieSceneScriptingFloatChannel, default=1, const [1,9]}
       pinky_03_l_ctrl.Location.X {
        type=MovieSceneScriptingFloatChannel
        default=-0.0002
        thinned_from=14
        keys [[2162,-0.0002],[2214,0],[2586,0],[2612,-0.0002],[2645,-0.0002],[2661,0],[2676,-0.0002],[2733,-0.0002],[2754,0],[2759,-0.0002]]
       }
       pinky_03_l_ctrl.Location.Y {
        type=MovieSceneScriptingFloatChannel
        default=0.0002
        thinned_from=14
        keys [[2162,0.0002],[2214,0],[2586,0],[2612,0.0002],[2645,0.0002],[2661,0],[2676,0.0002],[2733,0.0002],[2754,0],[2759,0.0002]]
       }
       pinky_03_l_ctrl.Location.Z {type=MovieSceneScriptingFloatChannel, default=0, thinned_from=14, keys [[2162,0],[2759,0]]}
       pinky_03_l_ctrl.Rotation.X {
        type=MovieSceneScriptingFloatChannel
        default=-0.2816
        thinned_from=16
        keys [[2162,-0.2816],[2202,0],[2586,0],[2612,-0.2816],[2645,-0.2816],[2661,0],[2665,-0.0001],[2676,-0.223],[2701,-0.223],[2708,-0.1966],[2720,-0.2816],[2733,-0.2816],[2747,-0.4944],[2754,-0.4944],[2759,0.1505]]
       }
       pinky_03_l_ctrl.Rotation.Y {
        type=MovieSceneScriptingFloatChannel
        default=1.0905
        thinned_from=16
        keys [[2162,1.0905],[2202,0],[2586,0],[2612,1.0905],[2645,1.0905],[2661,0],[2665,0.0006],[2676,0.8637],[2701,0.8637],[2708,0.7614],[2720,1.0905],[2733,1.0905],[2747,1.012],[2754,1.012],[2759,1.1162]]
       }
       pinky_03_l_ctrl.Rotation.Z {
        type=MovieSceneScriptingFloatChannel
        default=-5.0202
        thinned_from=16
        keys [[2162,-5.0202],[2202,-117.2954],[2586,-117.2954],[2612,-5.0202],[2645,-5.0202],[2661,-117.2954],[2665,-117.2383],[2676,-28.3725],[2701,-28.3725],[2708,-38.9067],[2720,-5.0202],[2733,-5.0202],[2747,-12.7965],[2754,-12.7965],[2759,17.1374]]
       }
       pinky_03_l_ctrl.Scale.X {type=MovieSceneScriptingFloatChannel, default=1, const [1,9]}
       pinky_03_l_ctrl.Scale.Y {type=MovieSceneScriptingFloatChannel, default=1, const [1,9]}
       pinky_03_l_ctrl.Scale.Z {type=MovieSceneScriptingFloatChannel, default=1, const [1,9]}
       hand_l_ik_ctrl.Location.X {type=MovieSceneScriptingFloatChannel, default=0, const [0,1021]}
       hand_l_ik_ctrl.Location.Y {type=MovieSceneScriptingFloatChannel, default=0, const [0,1021]}
       hand_l_ik_ctrl.Location.Z {type=MovieSceneScriptingFloatChannel, default=0, const [0,1021]}
       hand_l_ik_ctrl.Rotation.X {type=MovieSceneScriptingFloatChannel, default=0, const [0,1021]}
       hand_l_ik_ctrl.Rotation.Y {type=MovieSceneScriptingFloatChannel, default=0, const [0,1021]}
       hand_l_ik_ctrl.Rotation.Z {type=MovieSceneScriptingFloatChannel, default=0, const [0,1021]}
       hand_l_ik_ctrl.Scale.X {type=MovieSceneScriptingFloatChannel, default=1, const [1,1021]}
       hand_l_ik_ctrl.Scale.Y {type=MovieSceneScriptingFloatChannel, default=1, const [1,1021]}
       hand_l_ik_ctrl.Scale.Z {type=MovieSceneScriptingFloatChannel, default=1, const [1,1021]}
       Softness {type=MovieSceneScriptingFloatChannel, default=0, const [0,1021]}
       hand_l_ikorient_ctrl.Rotation.Roll {type=MovieSceneScriptingFloatChannel, default=0, const [0,1021]}
       hand_l_ikorient_ctrl.Rotation.Pitch {type=MovieSceneScriptingFloatChannel, default=0, const [0,1021]}
       hand_l_ikorient_ctrl.Rotation.Yaw {type=MovieSceneScriptingFloatChannel, default=0, const [0,1021]}
       arm_l_pv_ik_ctrl.Location.X {type=MovieSceneScriptingFloatChannel, default=0, const [0,1021]}
       arm_l_pv_ik_ctrl.Location.Y {type=MovieSceneScriptingFloatChannel, default=0, const [0,1021]}
       arm_l_pv_ik_ctrl.Location.Z {type=MovieSceneScriptingFloatChannel, default=0, const [0,1021]}
       arm_l_pv_ik_ctrl.Rotation.X {type=MovieSceneScriptingFloatChannel, default=0, const [0,1021]}
       arm_l_pv_ik_ctrl.Rotation.Y {type=MovieSceneScriptingFloatChannel, default=0, const [0,1021]}
       arm_l_pv_ik_ctrl.Rotation.Z {type=MovieSceneScriptingFloatChannel, default=0, const [0,1021]}
       arm_l_pv_ik_ctrl.Scale.X {type=MovieSceneScriptingFloatChannel, default=1, const [1,1021]}
       arm_l_pv_ik_ctrl.Scale.Y {type=MovieSceneScriptingFloatChannel, default=1, const [1,1021]}
       arm_l_pv_ik_ctrl.Scale.Z {type=MovieSceneScriptingFloatChannel, default=1, const [1,1021]}
       upperarm_r_fk_ctrl.Location.X {type=MovieSceneScriptingFloatChannel, default=0, const [0,19]}
       upperarm_r_fk_ctrl.Location.Y {type=MovieSceneScriptingFloatChannel, default=0, const [0,19]}
       upperarm_r_fk_ctrl.Location.Z {type=MovieSceneScriptingFloatChannel, default=0, const [0,19]}
       upperarm_r_fk_ctrl.Rotation.X {
        type=MovieSceneScriptingFloatChannel
        default=9.5916
        thinned_from=46
        keys [[2162,9.5916],[2180,9.5916],[2202,18.1075],[2208,2.4125],[2215,7.916],[2222,9.9272],[2230,9.9272],[2239,11.4212],[2266,4.8997],[2280,9.3189],[2290,9.665],[2307,10.029],[2316,10.0288],[2335,7.6774],[2352,9.9244],[2373,11.2608],[2391,10.8929],[2408,14.334],[2425,14.4347],[2444,14.3536],[2450,14.3066],[2461,11.7975],[2476,6.68],[2481,6.9711],[2493,9.9244],[2515,5.9743],[2551,8.9577],[2586,9.9244],[2604,-23.2357],[2624,-22.205],[2631,-22.0034],[2645,-22.1306],[2657,-18.0395],[2674,14.9369],[2683,14.7936],[2706,10.4613],[2712,-3.0895],[2726,-14.0053],[2733,-12.8961],[2744,50.0378],[2749,54.1717],[2754,54.1717],[2757,53.4373],[2759,69.3851]]
       }
       upperarm_r_fk_ctrl.Rotation.Y {
        type=MovieSceneScriptingFloatChannel
        default=-2.0299
        keys [[2162,-2.0299],[2180,-2.0299],[2202,7.7548],[2208,6.9281],[2215,3.0992],[2222,-1.6856],[2230,-1.6856],[2239,-0.3468],[2258,1.2641],[2266,0.8255],[2280,-1.375],[2290,-1.5702],[2307,2.734],[2316,2.7204],[2335,4.3535],[2352,-1.6885],[2373,0.0441],[2391,-0.5719],[2408,2.6797],[2425,1.2426],[2444,0.8786],[2450,0.7651],[2461,0.9278],[2476,1.2697],[2481,1.0042],[2493,-1.6885],[2515,1.6055],[2551,-0.8824],[2586,-1.6885],[2604,6.1306],[2624,-2.9126],[2631,-4.2373],[2642,-3.053],[2645,-3.0348],[2657,-1.7936],[2674,5.661],[2687,6.9101],[2706,7.6702],[2712,6.8118],[2726,9.4286],[2733,9.049],[2744,-1.7766],[2749,-2.5097],[2754,-2.5097],[2757,30.547],[2759,33.8828]]
       }
       upperarm_r_fk_ctrl.Rotation.Z {
        type=MovieSceneScriptingFloatChannel
        default=-8.6295
        keys [[2162,-8.6295],[2180,-8.6295],[2202,-28.8297],[2208,-59.6931],[2215,-30.558],[2222,-6.1782],[2230,-11.1751],[2239,-16.5828],[2258,-3.377],[2266,-0.0723],[2280,-19.0952],[2290,-17.0907],[2307,5.6252],[2316,5.5978],[2335,-18.3306],[2352,-11.1535],[2373,-19.0641],[2391,-14.8398],[2408,3.8204],[2425,11.3108],[2444,15.3044],[2450,15.2197],[2461,11.5383],[2476,2.5401],[2481,0.894],[2493,-11.1535],[2515,0.9208],[2551,-7.1583],[2586,-11.1535],[2604,-19.6597],[2624,-30.0874],[2631,-31.4648],[2642,-30.0369],[2645,-30.0128],[2657,-42.2458],[2674,-3.5055],[2683,4.2983],[2706,4.1946],[2726,-19.4827],[2733,-19.4827],[2744,3.9552],[2749,5.4886],[2754,5.4886],[2757,102.7717],[2759,170.5592]]
       }
       upperarm_r_fk_ctrl.Scale.X {type=MovieSceneScriptingFloatChannel, default=1, const [1,5]}
       upperarm_r_fk_ctrl.Scale.Y {type=MovieSceneScriptingFloatChannel, default=1, const [1,5]}
       upperarm_r_fk_ctrl.Scale.Z {type=MovieSceneScriptingFloatChannel, default=1, const [1,5]}
       lowerarm_r_fk_ctrl.Location.X {type=MovieSceneScriptingFloatChannel, default=0, const [0,16]}
       lowerarm_r_fk_ctrl.Location.Y {type=MovieSceneScriptingFloatChannel, default=0, const [0,16]}
       lowerarm_r_fk_ctrl.Location.Z {type=MovieSceneScriptingFloatChannel, default=0, const [0,17]}
       lowerarm_r_fk_ctrl.Rotation.X {type=MovieSceneScriptingFloatChannel, default=0, const [0,17]}
       lowerarm_r_fk_ctrl.Rotation.Y {type=MovieSceneScriptingFloatChannel, default=0, const [0,18]}
       lowerarm_r_fk_ctrl.Rotation.Z {
        type=MovieSceneScriptingFloatChannel
        default=28.0614
        thinned_from=19
        keys [[2162,28.0614],[2180,0],[2202,25.1837],[2208,10.2073],[2222,-4.7691],[2586,-4.7298],[2612,28.0614],[2656,28.0614],[2674,12.7433],[2706,12.7433],[2723,28.0614],[2733,28.0614],[2745,75.9909],[2749,79.6427],[2752,81.0189],[2754,81.1726],[2756,65.3337],[2759,-6.3206]]
       }
       lowerarm_r_fk_ctrl.Scale.X {type=MovieSceneScriptingFloatChannel, default=1, const [1,4]}
       lowerarm_r_fk_ctrl.Scale.Y {type=MovieSceneScriptingFloatChannel, default=1, const [1,4]}
       lowerarm_r_fk_ctrl.Scale.Z {type=MovieSceneScriptingFloatChannel, default=1, const [1,4]}
       hand_r_fk_ctrl.Location.X {type=MovieSceneScriptingFloatChannel, default=0, const [0,15]}
       hand_r_fk_ctrl.Location.Y {type=MovieSceneScriptingFloatChannel, default=0, const [0,15]}
       hand_r_fk_ctrl.Location.Z {type=MovieSceneScriptingFloatChannel, default=0, const [0,15]}
       hand_r_fk_ctrl.Rotation.X {
        type=MovieSceneScriptingFloatChannel
        default=8.3854
        thinned_from=18
        keys [[2162,8.3854],[2180,8.3854],[2202,-9.9207],[2208,-14.6201],[2222,-17.0731],[2586,-16.8636],[2612,8.0646],[2737,8.0646],[2747,-6.0998],[2754,-6.0998],[2757,-6.998],[2759,0]]
       }
       hand_r_fk_ctrl.Rotation.Y {
        type=MovieSceneScriptingFloatChannel
        default=-6.81
        thinned_from=18
        keys [[2162,-6.81],[2180,-6.81],[2202,11.8409],[2208,8.6984],[2222,5.5546],[2586,5.4498],[2612,5.8519],[2737,5.8519],[2747,18.2254],[2754,18.2254],[2757,2.7438],[2759,0]]
       }
       hand_r_fk_ctrl.Rotation.Z {
        type=MovieSceneScriptingFloatChannel
        default=0.7258
        thinned_from=18
        keys [[2162,0.7258],[2180,0.7258],[2202,-9.9831],[2208,-6.2141],[2222,-2.4446],[2586,-2.4177],[2612,-15.5419],[2737,-15.5419],[2747,-6.1348],[2754,-6.1348],[2757,-12.194],[2759,0]]
       }
       hand_r_fk_ctrl.Scale.X {type=MovieSceneScriptingFloatChannel, default=1, const [1,3]}
       hand_r_fk_ctrl.Scale.Y {type=MovieSceneScriptingFloatChannel, default=1, const [1,3]}
       hand_r_fk_ctrl.Scale.Z {type=MovieSceneScriptingFloatChannel, default=1, const [1,3]}
       thumb_01_r_ctrl.Location.X {type=MovieSceneScriptingFloatChannel, default=-0.0001, thinned_from=13, keys [[2162,-0.0001],[2759,-0.0001]]}
       thumb_01_r_ctrl.Location.Y {type=MovieSceneScriptingFloatChannel, default=0, thinned_from=13, keys [[2162,0],[2759,0]]}
       thumb_01_r_ctrl.Location.Z {type=MovieSceneScriptingFloatChannel, default=0, thinned_from=13, keys [[2162,0],[2759,0]]}
       thumb_01_r_ctrl.Rotation.X {
        type=MovieSceneScriptingFloatChannel
        default=6.0844
        thinned_from=18
        keys [[2162,6.0844],[2202,-1.5583],[2586,-1.5583],[2612,6.0844],[2627,5.1199],[2644,5.1844],[2652,6.0844],[2664,-1.5583],[2674,4.1029],[2702,4.1029],[2709,2.263],[2720,6.0844],[2731,6.0844],[2739,-17.8816],[2746,-1.5583],[2754,-1.5583],[2759,3.9725]]
       }
       thumb_01_r_ctrl.Rotation.Y {
        type=MovieSceneScriptingFloatChannel
        default=15.8669
        thinned_from=18
        keys [[2162,15.8669],[2202,0.8707],[2586,0.8707],[2612,15.8669],[2627,13.9744],[2644,14.101],[2652,15.8669],[2664,0.8707],[2674,11.979],[2702,11.979],[2709,8.3688],[2720,15.8669],[2731,15.8669],[2739,17.0028],[2746,0.8707],[2754,0.8707],[2759,19.3454]]
       }
       thumb_01_r_ctrl.Rotation.Z {
        type=MovieSceneScriptingFloatChannel
        default=-4.4052
        thinned_from=18
        keys [[2162,-4.4052],[2202,-13.9227],[2586,-13.9227],[2612,-4.4052],[2627,-5.6063],[2644,-5.5259],[2652,-4.4052],[2664,-13.9227],[2674,-6.8727],[2702,-6.8727],[2709,-9.1639],[2720,-4.4052],[2731,-4.4052],[2739,6.5079],[2746,-13.9227],[2754,-13.9227],[2759,-0.441]]
       }
       thumb_01_r_ctrl.Scale.X {type=MovieSceneScriptingFloatChannel, default=1, const [1,8]}
       thumb_01_r_ctrl.Scale.Y {type=MovieSceneScriptingFloatChannel, default=1, const [1,8]}
       thumb_01_r_ctrl.Scale.Z {type=MovieSceneScriptingFloatChannel, default=1, const [1,8]}
       thumb_02_r_ctrl.Location.X {
        type=MovieSceneScriptingFloatChannel
        default=0.0001
        thinned_from=13
        keys [[2162,0.0001],[2214,0],[2586,0],[2652,0.0001],[2664,0],[2754,0],[2759,0.0001]]
       }
       thumb_02_r_ctrl.Location.Y {type=MovieSceneScriptingFloatChannel, default=0, const [0,13]}
       thumb_02_r_ctrl.Location.Z {type=MovieSceneScriptingFloatChannel, default=0, const [0,13]}
       thumb_02_r_ctrl.Rotation.X {
        type=MovieSceneScriptingFloatChannel
        default=0.4753
        thinned_from=18
        keys [[2162,0.4753],[2202,0],[2586,0],[2612,0.4753],[2627,0.4153],[2644,0.4194],[2652,0.4753],[2664,0],[2674,0.3521],[2702,0.3521],[2709,0.2377],[2720,0.4753],[2731,0.4753],[2739,-2.2174],[2746,0],[2754,0],[2759,0.123]]
       }
       thumb_02_r_ctrl.Rotation.Y {
        type=MovieSceneScriptingFloatChannel
        default=1.8999
        thinned_from=18
        keys [[2162,1.8999],[2202,0],[2586,0],[2612,1.8999],[2627,1.6601],[2644,1.6762],[2652,1.8999],[2664,0],[2674,1.4073],[2702,1.4073],[2709,0.9499],[2720,1.8999],[2731,1.8999],[2739,5.1419],[2746,0],[2754,0],[2759,1.9494]]
       }
       thumb_02_r_ctrl.Rotation.Z {
        type=MovieSceneScriptingFloatChannel
        default=14.1339
        thinned_from=18
        keys [[2162,14.1339],[2202,0],[2586,0],[2612,14.1339],[2627,12.3502],[2644,12.4696],[2652,14.1339],[2664,0],[2674,10.4696],[2702,10.4696],[2709,7.0669],[2720,14.1339],[2731,14.1339],[2739,25.4436],[2746,0],[2754,0],[2759,3.6113]]
       }
       thumb_02_r_ctrl.Scale.X {type=MovieSceneScriptingFloatChannel, default=1, const [1,8]}
       thumb_02_r_ctrl.Scale.Y {type=MovieSceneScriptingFloatChannel, default=1, const [1,8]}
       thumb_02_r_ctrl.Scale.Z {type=MovieSceneScriptingFloatChannel, default=1, const [1,8]}
       thumb_03_r_ctrl.Location.X {type=MovieSceneScriptingFloatChannel, default=0, thinned_from=13, keys [[2162,0],[2759,0]]}
       thumb_03_r_ctrl.Location.Y {type=MovieSceneScriptingFloatChannel, default=0, thinned_from=13, keys [[2162,0],[2759,0]]}
       thumb_03_r_ctrl.Location.Z {type=MovieSceneScriptingFloatChannel, default=0, thinned_from=13, keys [[2162,0],[2759,0]]}
       thumb_03_r_ctrl.Rotation.X {
        type=MovieSceneScriptingFloatChannel
        default=1.3634
        thinned_from=18
        keys [[2162,1.3634],[2202,8.1645],[2586,8.1645],[2612,1.3634],[2627,2.2217],[2644,2.1643],[2652,1.3634],[2664,8.1645],[2674,3.1267],[2702,3.1267],[2709,4.764],[2720,1.3634],[2731,1.3634],[2739,-4.1701],[2746,8.1645],[2754,8.1645],[2759,0.6645]]
       }
       thumb_03_r_ctrl.Rotation.Y {
        type=MovieSceneScriptingFloatChannel
        default=-1.4762
        thinned_from=18
        keys [[2162,-1.4762],[2202,-9.9357],[2586,-9.9357],[2612,-1.4762],[2627,-2.5438],[2644,-2.4723],[2652,-1.4762],[2664,-9.9357],[2674,-3.6694],[2702,-3.6694],[2709,-5.706],[2720,-1.4762],[2731,-1.4762],[2739,-0.5827],[2746,-9.9357],[2754,-9.9357],[2759,-1.8964]]
       }
       thumb_03_r_ctrl.Rotation.Z {
        type=MovieSceneScriptingFloatChannel
        default=0
        thinned_from=18
        keys [[2162,0],[2202,-50.7089],[2586,-50.7089],[2612,0],[2627,-6.3995],[2644,-5.9712],[2652,0],[2664,-50.7089],[2674,-13.1467],[2702,-13.1467],[2709,-25.3544],[2720,0],[2731,0],[2739,2.705],[2746,-50.7089],[2754,-50.7089],[2759,23.4201]]
       }
       thumb_03_r_ctrl.Scale.X {type=MovieSceneScriptingFloatChannel, default=1, const [1,8]}
       thumb_03_r_ctrl.Scale.Y {type=MovieSceneScriptingFloatChannel, default=1, const [1,8]}
       thumb_03_r_ctrl.Scale.Z {type=MovieSceneScriptingFloatChannel, default=1, const [1,8]}
       index_metacarpal_r_ctrl.Location.X {type=MovieSceneScriptingFloatChannel, default=0.0001, thinned_from=13, keys [[2162,0.0001],[2759,0.0001]]}
       index_metacarpal_r_ctrl.Location.Y {type=MovieSceneScriptingFloatChannel, default=0.0002, thinned_from=13, keys [[2162,0.0002],[2731,0.0002],[2739,0.0001],[2759,0.0002]]}
       index_metacarpal_r_ctrl.Location.Z {type=MovieSceneScriptingFloatChannel, default=0, thinned_from=13, keys [[2162,0],[2759,0]]}
       index_metacarpal_r_ctrl.Rotation.X {type=MovieSceneScriptingFloatChannel, default=0, const [0,18]}
       index_metacarpal_r_ctrl.Rotation.Y {
        type=MovieSceneScriptingFloatChannel
        default=0
        thinned_from=18
        keys [[2162,0],[2202,3.7815],[2586,3.7815],[2612,0],[2627,0.4772],[2644,0.4453],[2652,0],[2664,3.7815],[2674,0.9804],[2702,0.9804],[2709,1.8908],[2720,0],[2739,0],[2746,3.7815],[2754,3.7815],[2759,1.7196]]
       }
       index_metacarpal_r_ctrl.Rotation.Z {
        type=MovieSceneScriptingFloatChannel
        default=0
        thinned_from=18
        keys [[2162,0],[2202,-3.094],[2586,-3.094],[2612,0],[2627,-0.3905],[2644,-0.3643],[2652,0],[2664,-3.094],[2674,-0.8021],[2702,-0.8021],[2709,-1.547],[2720,0],[2739,0],[2746,-3.094],[2754,-3.094],[2759,0]]
       }
       index_metacarpal_r_ctrl.Scale.X {type=MovieSceneScriptingFloatChannel, default=1, const [1,8]}
       index_metacarpal_r_ctrl.Scale.Y {type=MovieSceneScriptingFloatChannel, default=1, const [1,8]}
       index_metacarpal_r_ctrl.Scale.Z {type=MovieSceneScriptingFloatChannel, default=1, const [1,8]}
       index_01_r_ctrl.Location.X {type=MovieSceneScriptingFloatChannel, default=0, thinned_from=13, keys [[2162,0],[2759,0]]}
       index_01_r_ctrl.Location.Y {type=MovieSceneScriptingFloatChannel, default=0, thinned_from=13, keys [[2162,0],[2759,0]]}
       index_01_r_ctrl.Location.Z {type=MovieSceneScriptingFloatChannel, default=0, thinned_from=13, keys [[2162,0],[2759,0]]}
       index_01_r_ctrl.Rotation.X {
        type=MovieSceneScriptingFloatChannel
        default=0
        thinned_from=18
        keys [[2162,0],[2202,6.2821],[2586,6.2821],[2612,0],[2627,0.7928],[2644,0.7397],[2652,0],[2664,6.2821],[2674,1.6287],[2702,1.6287],[2709,3.1411],[2720,0],[2731,0],[2739,1.9817],[2746,6.2821],[2754,6.2821],[2759,-4.7213]]
       }
       index_01_r_ctrl.Rotation.Y {
        type=MovieSceneScriptingFloatChannel
        default=0
        thinned_from=18
        keys [[2162,0],[2202,-2.5276],[2586,-2.5276],[2612,0],[2627,-0.319],[2644,-0.2976],[2652,0],[2664,-2.5276],[2674,-0.6553],[2702,-0.6553],[2709,-1.2638],[2720,0],[2731,0],[2739,-4.3553],[2746,-2.5276],[2754,-2.5276],[2759,7.9585]]
       }
       index_01_r_ctrl.Rotation.Z {
        type=MovieSceneScriptingFloatChannel
        default=7.2214
        thinned_from=18
        keys [[2162,7.2214],[2202,-65.1155],[2586,-65.1155],[2612,7.2214],[2627,-1.9075],[2644,-1.2966],[2652,7.2214],[2664,-65.1155],[2674,-11.5326],[2702,-11.5326],[2709,-28.947],[2720,7.2214],[2731,7.2214],[2739,7.9953],[2746,-65.1155],[2754,-65.1155],[2759,31.1517]]
       }
       index_01_r_ctrl.Scale.X {type=MovieSceneScriptingFloatChannel, default=1, const [1,8]}
       index_01_r_ctrl.Scale.Y {type=MovieSceneScriptingFloatChannel, default=1, const [1,8]}
       index_01_r_ctrl.Scale.Z {type=MovieSceneScriptingFloatChannel, default=1, const [1,8]}
       index_02_r_ctrl.Location.X {type=MovieSceneScriptingFloatChannel, default=0, thinned_from=13, keys [[2162,0],[2759,0]]}
       index_02_r_ctrl.Location.Y {type=MovieSceneScriptingFloatChannel, default=-0.0001, thinned_from=13, keys [[2162,-0.0001],[2759,-0.0001]]}
       index_02_r_ctrl.Location.Z {type=MovieSceneScriptingFloatChannel, default=0, thinned_from=13, keys [[2162,0],[2759,0]]}
       index_02_r_ctrl.Rotation.X {
        type=MovieSceneScriptingFloatChannel
        default=-0.23
        thinned_from=18
        keys [[2162,-0.23],[2202,0],[2586,0],[2612,-0.23],[2627,-0.201],[2644,-0.203],[2652,-0.23],[2664,0],[2674,-0.1704],[2702,-0.1704],[2709,-0.115],[2720,-0.23],[2731,-0.23],[2739,-0.0908],[2746,0],[2754,0],[2759,-0.1354]]
       }
       index_02_r_ctrl.Rotation.Y {
        type=MovieSceneScriptingFloatChannel
        default=0.3153
        thinned_from=18
        keys [[2162,0.3153],[2202,0],[2586,0],[2612,0.3153],[2627,0.2755],[2644,0.2782],[2652,0.3153],[2664,0],[2674,0.2336],[2702,0.2336],[2709,0.1577],[2720,0.3153],[2731,0.3153],[2739,1.1648],[2746,0],[2754,0],[2759,3.7485]]
       }
       index_02_r_ctrl.Rotation.Z {
        type=MovieSceneScriptingFloatChannel
        default=20.8963
        thinned_from=18
        keys [[2162,20.8963],[2202,-60.8298],[2586,-60.8298],[2612,20.8963],[2627,10.5825],[2644,11.2727],[2652,20.8963],[2664,-60.8298],[2674,-0.2919],[2702,-0.2919],[2709,-19.9667],[2720,20.8963],[2731,20.8963],[2739,27.0212],[2746,-60.8298],[2754,-60.8298],[2759,35.3799]]
       }
       index_02_r_ctrl.Scale.X {type=MovieSceneScriptingFloatChannel, default=1, const [1,8]}
       index_02_r_ctrl.Scale.Y {type=MovieSceneScriptingFloatChannel, default=1, const [1,8]}
       index_02_r_ctrl.Scale.Z {type=MovieSceneScriptingFloatChannel, default=1, const [1,8]}
       index_03_r_ctrl.Location.X {type=MovieSceneScriptingFloatChannel, default=0, thinned_from=13, keys [[2162,0],[2759,0]]}
       index_03_r_ctrl.Location.Y {type=MovieSceneScriptingFloatChannel, default=0, thinned_from=13, keys [[2162,0],[2759,0]]}
       index_03_r_ctrl.Location.Z {type=MovieSceneScriptingFloatChannel, default=0, thinned_from=13, keys [[2162,0],[2759,0]]}
       index_03_r_ctrl.Rotation.X {
        type=MovieSceneScriptingFloatChannel
        default=-0.0867
        thinned_from=18
        keys [[2162,-0.0867],[2202,0],[2586,0],[2612,-0.0867],[2627,-0.0757],[2644,-0.0765],[2652,-0.0867],[2664,0],[2674,-0.0642],[2702,-0.0642],[2709,-0.0433],[2720,-0.0867],[2731,-0.0867],[2739,-0.0342],[2746,0],[2754,0],[2759,-0.0057]]
       }
       index_03_r_ctrl.Rotation.Y {
        type=MovieSceneScriptingFloatChannel
        default=0.7964
        thinned_from=18
        keys [[2162,0.7964],[2202,0],[2586,0],[2612,0.7964],[2627,0.6959],[2644,0.7027],[2652,0.7964],[2664,0],[2674,0.59],[2702,0.59],[2709,0.3982],[2720,0.7964],[2731,0.7964],[2739,1.3547],[2746,0],[2754,0],[2759,0.8011]]
       }
       index_03_r_ctrl.Rotation.Z {
        type=MovieSceneScriptingFloatChannel
        default=25.1985
        thinned_from=18
        keys [[2162,25.1985],[2202,-78.806],[2586,-78.806],[2612,25.1985],[2627,12.0731],[2644,12.9515],[2652,25.1985],[2664,-78.806],[2674,-1.7656],[2702,-1.7656],[2709,-26.8038],[2720,25.1985],[2739,25.2392],[2746,-78.806],[2754,-78.806],[2759,31.0033]]
       }
       index_03_r_ctrl.Scale.X {type=MovieSceneScriptingFloatChannel, default=1, const [1,8]}
       index_03_r_ctrl.Scale.Y {type=MovieSceneScriptingFloatChannel, default=1, const [1,8]}
       index_03_r_ctrl.Scale.Z {type=MovieSceneScriptingFloatChannel, default=1, const [1,8]}
       middle_metacarpal_r_ctrl.Location.X {type=MovieSceneScriptingFloatChannel, default=0, thinned_from=13, keys [[2162,0],[2759,0]]}
       middle_metacarpal_r_ctrl.Location.Y {type=MovieSceneScriptingFloatChannel, default=0.0001, thinned_from=13, keys [[2162,0.0001],[2759,0.0001]]}
       middle_metacarpal_r_ctrl.Location.Z {type=MovieSceneScriptingFloatChannel, default=0, thinned_from=13, keys [[2162,0],[2759,0]]}
       middle_metacarpal_r_ctrl.Rotation.X {
        type=MovieSceneScriptingFloatChannel
        default=0
        thinned_from=18
        keys [[2162,0],[2202,-7.29],[2586,-7.29],[2612,0],[2627,-0.92],[2644,-0.8584],[2652,0],[2664,-7.29],[2674,-1.89],[2702,-1.89],[2709,-3.645],[2720,0],[2739,0],[2746,-7.29],[2754,-7.29],[2759,0]]
       }
       middle_metacarpal_r_ctrl.Rotation.Y {
        type=MovieSceneScriptingFloatChannel
        default=0
        thinned_from=18
        keys [[2162,0],[2202,1.3582],[2586,1.3582],[2612,0],[2627,0.1714],[2644,0.1599],[2652,0],[2664,1.3582],[2674,0.3521],[2702,0.3521],[2709,0.6791],[2720,0],[2731,0],[2739,0.5549],[2746,1.3582],[2754,1.3582],[2759,0]]
       }
       middle_metacarpal_r_ctrl.Rotation.Z {
        type=MovieSceneScriptingFloatChannel
        default=0
        thinned_from=18
        keys [[2162,0],[2202,-2.7048],[2586,-2.7048],[2612,0],[2627,-0.3413],[2644,-0.3185],[2652,0],[2664,-2.7048],[2674,-0.7013],[2702,-0.7013],[2709,-1.3524],[2720,0],[2739,0],[2746,-2.7048],[2754,-2.7048],[2759,0]]
       }
       middle_metacarpal_r_ctrl.Scale.X {type=MovieSceneScriptingFloatChannel, default=1, const [1,8]}
       middle_metacarpal_r_ctrl.Scale.Y {type=MovieSceneScriptingFloatChannel, default=1, const [1,8]}
       middle_metacarpal_r_ctrl.Scale.Z {type=MovieSceneScriptingFloatChannel, default=1, const [1,8]}
       middle_01_r_ctrl.Location.X {type=MovieSceneScriptingFloatChannel, default=0.0002, thinned_from=13, keys [[2162,0.0002],[2731,0.0002],[2739,0.0001],[2759,0.0002]]}
       middle_01_r_ctrl.Location.Y {type=MovieSceneScriptingFloatChannel, default=0, thinned_from=13, keys [[2162,0],[2759,0]]}
       middle_01_r_ctrl.Location.Z {type=MovieSceneScriptingFloatChannel, default=0, thinned_from=13, keys [[2162,0],[2759,0]]}
       middle_01_r_ctrl.Rotation.X {type=MovieSceneScriptingFloatChannel, default=0, thinned_from=18, keys [[2162,0],[2731,0],[2739,-1.514],[2746,0],[2754,0],[2759,-3.9745]]}
       middle_01_r_ctrl.Rotation.Y {
        type=MovieSceneScriptingFloatChannel
        default=0
        thinned_from=18
        keys [[2162,0],[2202,1.3494],[2586,1.3494],[2612,0],[2627,0.1703],[2644,0.1589],[2652,0],[2664,1.3494],[2674,0.3499],[2702,0.3499],[2709,0.6747],[2720,0],[2731,0],[2739,-1.9243],[2746,1.3494],[2754,1.3494],[2759,0]]
       }
       middle_01_r_ctrl.Rotation.Z {
        type=MovieSceneScriptingFloatChannel
        default=3.8267
        thinned_from=18
        keys [[2162,3.8267],[2202,-74.7137],[2586,-74.7137],[2612,3.8267],[2627,-6.0851],[2644,-5.4219],[2652,3.8267],[2664,-74.7137],[2674,-16.5356],[2702,-16.5356],[2709,-35.4435],[2720,3.8267],[2731,3.8267],[2739,8.789],[2746,-74.7137],[2754,-74.7137],[2759,18.3246]]
       }
       middle_01_r_ctrl.Scale.X {type=MovieSceneScriptingFloatChannel, default=1, const [1,8]}
       middle_01_r_ctrl.Scale.Y {type=MovieSceneScriptingFloatChannel, default=1, const [1,8]}
       middle_01_r_ctrl.Scale.Z {type=MovieSceneScriptingFloatChannel, default=1, const [1,8]}
       middle_02_r_ctrl.Location.X {
        type=MovieSceneScriptingFloatChannel
        default=0.0008
        thinned_from=13
        keys [[2162,0.0008],[2731,0.0008],[2739,0.0003],[2746,0.0008],[2759,0.0008]]
       }
       middle_02_r_ctrl.Location.Y {
        type=MovieSceneScriptingFloatChannel
        default=0.0009
        thinned_from=13
        keys [[2162,0.0009],[2731,0.0009],[2739,0.0003],[2746,0.0009],[2759,0.0009]]
       }
       middle_02_r_ctrl.Location.Z {type=MovieSceneScriptingFloatChannel, default=0.0476, const [0.0476,13]}
       middle_02_r_ctrl.Rotation.X {
        type=MovieSceneScriptingFloatChannel
        default=-2.9188
        thinned_from=18
        keys [[2162,-2.9188],[2202,0.0065],[2586,0.0065],[2612,-2.9188],[2627,-2.5496],[2644,-2.5743],[2652,-2.9188],[2664,0.0065],[2674,-2.1603],[2702,-2.1603],[2709,-1.4561],[2720,-2.9188],[2731,-2.9188],[2739,-1.1586],[2746,0.0065],[2754,0.0065],[2759,-1.4489]]
       }
       middle_02_r_ctrl.Rotation.Y {
        type=MovieSceneScriptingFloatChannel
        default=2.5438
        thinned_from=18
        keys [[2162,2.5438],[2202,-0.0119],[2586,-0.0119],[2612,2.5438],[2627,2.2213],[2644,2.2429],[2652,2.5438],[2664,-0.0119],[2674,1.8812],[2702,1.8812],[2709,1.266],[2720,2.5438],[2731,2.5438],[2739,0.9991],[2746,-0.0119],[2754,-0.0119],[2759,6.9733]]
       }
       middle_02_r_ctrl.Rotation.Z {
        type=MovieSceneScriptingFloatChannel
        default=12.0645
        thinned_from=18
        keys [[2162,12.0645],[2202,-58.0462],[2586,-58.0462],[2612,12.0645],[2627,3.2165],[2644,3.8086],[2652,12.0645],[2664,-58.0462],[2674,-6.1123],[2702,-6.1123],[2709,-22.9908],[2720,12.0645],[2731,12.0645],[2739,18.3197],[2746,-58.0462],[2754,-58.0462],[2759,39.0682]]
       }
       middle_02_r_ctrl.Scale.X {type=MovieSceneScriptingFloatChannel, default=1, const [1,8]}
       middle_02_r_ctrl.Scale.Y {type=MovieSceneScriptingFloatChannel, default=1, const [1,8]}
       middle_02_r_ctrl.Scale.Z {type=MovieSceneScriptingFloatChannel, default=1, const [1,8]}
       middle_03_r_ctrl.Location.X {
        type=MovieSceneScriptingFloatChannel
        default=0.0005
        thinned_from=13
        keys [[2162,0.0005],[2731,0.0005],[2739,0.0002],[2746,0.0005],[2759,0.0005]]
       }
       middle_03_r_ctrl.Location.Y {type=MovieSceneScriptingFloatChannel, default=0, thinned_from=13, keys [[2162,0],[2759,0]]}
       middle_03_r_ctrl.Location.Z {type=MovieSceneScriptingFloatChannel, default=0.0061, const [0.0061,13]}
       middle_03_r_ctrl.Rotation.X {
        type=MovieSceneScriptingFloatChannel
        default=1.0874
        thinned_from=18
        keys [[2162,1.0874],[2202,-0.007],[2586,-0.007],[2612,1.0874],[2627,0.9493],[2644,0.9585],[2652,1.0874],[2664,-0.007],[2674,0.8037],[2702,0.8037],[2709,0.5402],[2720,1.0874],[2731,1.0874],[2739,0.4241],[2746,-0.007],[2754,-0.007],[2759,1.6933]]
       }
       middle_03_r_ctrl.Rotation.Y {
        type=MovieSceneScriptingFloatChannel
        default=2.0851
        thinned_from=18
        keys [[2162,2.0851],[2202,-0.0086],[2586,-0.0086],[2612,2.0851],[2627,1.8209],[2644,1.8386],[2652,2.0851],[2664,-0.0086],[2674,1.5423],[2702,1.5423],[2709,1.0383],[2720,2.0851],[2731,2.0851],[2739,0.8274],[2746,-0.0086],[2754,-0.0086],[2759,1.6319]]
       }
       middle_03_r_ctrl.Rotation.Z {
        type=MovieSceneScriptingFloatChannel
        default=14.0525
        thinned_from=18
        keys [[2162,14.0525],[2202,-64.5776],[2586,-64.5776],[2612,14.0525],[2627,4.1294],[2644,4.7934],[2652,14.0525],[2664,-64.5776],[2674,-6.3331],[2702,-6.3331],[2709,-25.2626],[2720,14.0525],[2731,14.0525],[2739,21.4036],[2746,-64.5776],[2754,-64.5776],[2759,32.5715]]
       }
       middle_03_r_ctrl.Scale.X {type=MovieSceneScriptingFloatChannel, default=1, const [1,8]}
       middle_03_r_ctrl.Scale.Y {type=MovieSceneScriptingFloatChannel, default=1, const [1,8]}
       middle_03_r_ctrl.Scale.Z {type=MovieSceneScriptingFloatChannel, default=1, const [1,8]}
       ring_metacarpal_r_ctrl.Location.X {type=MovieSceneScriptingFloatChannel, default=0.0001, thinned_from=13, keys [[2162,0.0001],[2759,0.0001]]}
       ring_metacarpal_r_ctrl.Location.Y {type=MovieSceneScriptingFloatChannel, default=0.0002, thinned_from=13, keys [[2162,0.0002],[2759,0.0002]]}
       ring_metacarpal_r_ctrl.Location.Z {type=MovieSceneScriptingFloatChannel, default=0, thinned_from=13, keys [[2162,0],[2759,0]]}
       ring_metacarpal_r_ctrl.Rotation.X {
        type=MovieSceneScriptingFloatChannel
        default=0
        thinned_from=18
        keys [[2162,0],[2202,-2.4847],[2586,-2.4847],[2612,0],[2627,-0.3136],[2644,-0.2926],[2652,0],[2664,-2.4847],[2674,-0.6442],[2702,-0.6442],[2709,-1.2424],[2720,0],[2739,0],[2746,-2.4847],[2754,-2.4847],[2759,0]]
       }
       ring_metacarpal_r_ctrl.Rotation.Y {
        type=MovieSceneScriptingFloatChannel
        default=0
        thinned_from=18
        keys [[2162,0],[2202,-0.8637],[2586,-0.8637],[2612,0],[2627,-0.109],[2644,-0.1017],[2652,0],[2664,-0.8637],[2674,-0.2239],[2702,-0.2239],[2709,-0.4318],[2720,0],[2731,0],[2739,-1.5259],[2746,-0.8637],[2754,-0.8637],[2759,-1.1632]]
       }
       ring_metacarpal_r_ctrl.Rotation.Z {
        type=MovieSceneScriptingFloatChannel
        default=0
        thinned_from=18
        keys [[2162,0],[2202,-5.6698],[2586,-5.6698],[2612,0],[2627,-0.7155],[2644,-0.6676],[2652,0],[2664,-5.6698],[2674,-1.4699],[2702,-1.4699],[2709,-2.8349],[2720,0],[2731,0],[2739,-1.3872],[2746,-5.6698],[2754,-5.6698],[2759,0]]
       }
       ring_metacarpal_r_ctrl.Scale.X {type=MovieSceneScriptingFloatChannel, default=1, const [1,8]}
       ring_metacarpal_r_ctrl.Scale.Y {type=MovieSceneScriptingFloatChannel, default=1, const [1,8]}
       ring_metacarpal_r_ctrl.Scale.Z {type=MovieSceneScriptingFloatChannel, default=1, const [1,8]}
       ring_01_r_ctrl.Location.X {type=MovieSceneScriptingFloatChannel, default=-0.0002, thinned_from=13, keys [[2162,-0.0002],[2731,-0.0002],[2739,-0.0001],[2759,-0.0002]]}
       ring_01_r_ctrl.Location.Y {type=MovieSceneScriptingFloatChannel, default=0, thinned_from=13, keys [[2162,0],[2759,0]]}
       ring_01_r_ctrl.Location.Z {type=MovieSceneScriptingFloatChannel, default=0, thinned_from=13, keys [[2162,0],[2759,0]]}
       ring_01_r_ctrl.Rotation.X {
        type=MovieSceneScriptingFloatChannel
        default=0
        thinned_from=18
        keys [[2162,0],[2202,-4.509],[2586,-4.509],[2612,0],[2627,-0.569],[2644,-0.531],[2652,0],[2664,-4.509],[2674,-1.169],[2702,-1.169],[2709,-2.2545],[2720,0],[2731,0],[2739,-5.8882],[2746,-4.509],[2754,-4.509],[2759,10.5727]]
       }
       ring_01_r_ctrl.Rotation.Y {
        type=MovieSceneScriptingFloatChannel
        default=-0.3119
        thinned_from=18
        keys [[2162,-0.3119],[2202,-1.201],[2586,-1.201],[2612,-0.3119],[2627,-0.4241],[2644,-0.4166],[2652,-0.3119],[2664,-1.201],[2674,-0.5424],[2702,-0.5424],[2709,-0.7564],[2720,-0.3119],[2731,-0.3119],[2739,-1.6031],[2746,-1.201],[2754,-1.201],[2759,-4.6626]]
       }
       ring_01_r_ctrl.Rotation.Z {
        type=MovieSceneScriptingFloatChannel
        default=0
        thinned_from=18
        keys [[2162,0],[2202,-79.0231],[2586,-79.0231],[2612,0],[2627,-9.9727],[2644,-9.3054],[2652,0],[2664,-79.0231],[2674,-20.4875],[2702,-20.4875],[2709,-39.5115],[2720,0],[2731,0],[2739,12.7651],[2746,-79.0231],[2754,-79.0231],[2759,22.2578]]
       }
       ring_01_r_ctrl.Scale.X {type=MovieSceneScriptingFloatChannel, default=1, const [1,8]}
       ring_01_r_ctrl.Scale.Y {type=MovieSceneScriptingFloatChannel, default=1, const [1,8]}
       ring_01_r_ctrl.Scale.Z {type=MovieSceneScriptingFloatChannel, default=1, const [1,8]}
       ring_02_r_ctrl.Location.X {type=MovieSceneScriptingFloatChannel, default=0.0003, thinned_from=13, keys [[2162,0.0003],[2731,0.0003],[2739,0.0001],[2759,0.0003]]}
       ring_02_r_ctrl.Location.Y {type=MovieSceneScriptingFloatChannel, default=0, thinned_from=13, keys [[2162,0],[2759,0]]}
       ring_02_r_ctrl.Location.Z {type=MovieSceneScriptingFloatChannel, default=0, thinned_from=13, keys [[2162,0],[2759,0]]}
       ring_02_r_ctrl.Rotation.X {type=MovieSceneScriptingFloatChannel, default=0, thinned_from=18, keys [[2162,0],[2759,0]]}
       ring_02_r_ctrl.Rotation.Y {type=MovieSceneScriptingFloatChannel, default=0, thinned_from=18, keys [[2162,0],[2754,0],[2759,3.3812]]}
       ring_02_r_ctrl.Rotation.Z {
        type=MovieSceneScriptingFloatChannel
        default=0
        thinned_from=18
        keys [[2162,0],[2202,-66.6687],[2586,-66.6687],[2612,0],[2627,-8.4136],[2644,-7.8506],[2652,0],[2664,-66.6687],[2674,-17.2845],[2702,-17.2845],[2709,-33.3344],[2720,0],[2731,0],[2739,11.4586],[2746,-66.6687],[2754,-66.6687],[2759,35.2878]]
       }
       ring_02_r_ctrl.Scale.X {type=MovieSceneScriptingFloatChannel, default=1, const [1,8]}
       ring_02_r_ctrl.Scale.Y {type=MovieSceneScriptingFloatChannel, default=1, const [1,8]}
       ring_02_r_ctrl.Scale.Z {type=MovieSceneScriptingFloatChannel, default=1, const [1,8]}
       ring_03_r_ctrl.Location.X {type=MovieSceneScriptingFloatChannel, default=0, thinned_from=13, keys [[2162,0],[2759,0]]}
       ring_03_r_ctrl.Location.Y {type=MovieSceneScriptingFloatChannel, default=0, thinned_from=13, keys [[2162,0],[2759,0]]}
       ring_03_r_ctrl.Location.Z {type=MovieSceneScriptingFloatChannel, default=0, thinned_from=13, keys [[2162,0],[2759,0]]}
       ring_03_r_ctrl.Rotation.X {type=MovieSceneScriptingFloatChannel, default=0, thinned_from=18, keys [[2162,0],[2759,0]]}
       ring_03_r_ctrl.Rotation.Y {type=MovieSceneScriptingFloatChannel, default=0, thinned_from=18, keys [[2162,0],[2759,0]]}
       ring_03_r_ctrl.Rotation.Z {
        type=MovieSceneScriptingFloatChannel
        default=0
        thinned_from=18
        keys [[2162,0],[2202,-73.2004],[2586,-73.2004],[2612,0],[2627,-9.2379],[2644,-8.6197],[2652,0],[2664,-73.2004],[2674,-18.9779],[2702,-18.9779],[2709,-36.6002],[2720,0],[2731,0],[2739,10.1876],[2746,-73.2004],[2754,-73.2004],[2759,20.858]]
       }
       ring_03_r_ctrl.Scale.X {type=MovieSceneScriptingFloatChannel, default=1, const [1,8]}
       ring_03_r_ctrl.Scale.Y {type=MovieSceneScriptingFloatChannel, default=1, const [1,8]}
       ring_03_r_ctrl.Scale.Z {type=MovieSceneScriptingFloatChannel, default=1, const [1,8]}
       pinky_metacarpal_r_ctrl.Location.X {
        type=MovieSceneScriptingFloatChannel
        default=0
        thinned_from=13
        keys [[2162,0],[2214,-0.0001],[2586,-0.0001],[2652,0],[2664,-0.0001],[2739,0],[2754,-0.0001],[2759,0]]
       }
       pinky_metacarpal_r_ctrl.Location.Y {type=MovieSceneScriptingFloatChannel, default=0, thinned_from=13, keys [[2162,0],[2759,0]]}
       pinky_metacarpal_r_ctrl.Location.Z {type=MovieSceneScriptingFloatChannel, default=0, thinned_from=13, keys [[2162,0],[2759,0]]}
       pinky_metacarpal_r_ctrl.Rotation.X {
        type=MovieSceneScriptingFloatChannel
        default=0
        thinned_from=18
        keys [[2162,0],[2202,-5.9588],[2586,-5.9588],[2612,0],[2627,-0.752],[2644,-0.7017],[2652,0],[2664,-5.9588],[2674,-1.5449],[2702,-1.5449],[2709,-2.9794],[2720,0],[2739,0],[2746,-5.9588],[2754,-5.9588],[2759,0]]
       }
       pinky_metacarpal_r_ctrl.Rotation.Y {
        type=MovieSceneScriptingFloatChannel
        default=-3.1934
        thinned_from=18
        keys [[2162,-3.1934],[2202,-5.2205],[2586,-5.2205],[2612,-3.1934],[2627,-3.4492],[2644,-3.4321],[2652,-3.1934],[2664,-5.2205],[2674,-3.7189],[2702,-3.7189],[2709,-4.2069],[2720,-3.1934],[2731,-3.1934],[2739,-2.3703],[2746,-5.2205],[2754,-5.2205],[2759,-3.1934]]
       }
       pinky_metacarpal_r_ctrl.Rotation.Z {
        type=MovieSceneScriptingFloatChannel
        default=0
        thinned_from=18
        keys [[2162,0],[2202,-11.0008],[2586,-11.0008],[2612,0],[2627,-1.3883],[2644,-1.2954],[2652,0],[2664,-11.0008],[2674,-2.8521],[2702,-2.8521],[2709,-5.5004],[2720,0],[2731,0],[2739,-2.1501],[2746,-11.0008],[2754,-11.0008],[2759,0]]
       }
       pinky_metacarpal_r_ctrl.Scale.X {type=MovieSceneScriptingFloatChannel, default=1, const [1,8]}
       pinky_metacarpal_r_ctrl.Scale.Y {type=MovieSceneScriptingFloatChannel, default=1, const [1,8]}
       pinky_metacarpal_r_ctrl.Scale.Z {type=MovieSceneScriptingFloatChannel, default=1, const [1,8]}
       pinky_01_r_ctrl.Location.X {type=MovieSceneScriptingFloatChannel, default=0.0001, thinned_from=13, keys [[2162,0.0001],[2759,0.0001]]}
       pinky_01_r_ctrl.Location.Y {type=MovieSceneScriptingFloatChannel, default=0.0002, thinned_from=13, keys [[2162,0.0002],[2731,0.0002],[2739,0.0001],[2759,0.0002]]}
       pinky_01_r_ctrl.Location.Z {type=MovieSceneScriptingFloatChannel, default=0, thinned_from=13, keys [[2162,0],[2759,0]]}
       pinky_01_r_ctrl.Rotation.X {
        type=MovieSceneScriptingFloatChannel
        default=-4.1261
        thinned_from=18
        keys [[2162,-4.1261],[2202,-11.7646],[2586,-11.7646],[2612,-4.1261],[2627,-5.0901],[2644,-5.0256],[2652,-4.1261],[2664,-11.7646],[2674,-6.1065],[2702,-6.1065],[2709,-7.9454],[2720,-4.1261],[2731,-4.1261],[2739,-15.632],[2746,-11.7646],[2754,-11.7646],[2759,14.6395]]
       }
       pinky_01_r_ctrl.Rotation.Y {
        type=MovieSceneScriptingFloatChannel
        default=1.0722
        thinned_from=18
        keys [[2162,1.0722],[2202,-1.0914],[2586,-1.0914],[2612,1.0722],[2627,0.7992],[2644,0.8174],[2652,1.0722],[2664,-1.0914],[2674,0.5113],[2702,0.5113],[2709,-0.0096],[2720,1.0722],[2731,1.0722],[2739,-3.5844],[2746,-1.0914],[2754,-1.0914],[2759,-11.264]]
       }
       pinky_01_r_ctrl.Rotation.Z {
        type=MovieSceneScriptingFloatChannel
        default=-14.4594
        thinned_from=18
        keys [[2162,-14.4594],[2202,-84.6646],[2586,-84.6646],[2612,-14.4594],[2627,-23.3193],[2644,-22.7264],[2652,-14.4594],[2664,-84.6646],[2674,-32.6608],[2702,-32.6608],[2709,-49.562],[2720,-14.4594],[2731,-14.4594],[2739,1.9945],[2746,-84.6646],[2754,-84.6646],[2759,24.5562]]
       }
       pinky_01_r_ctrl.Scale.X {type=MovieSceneScriptingFloatChannel, default=1, const [1,8]}
       pinky_01_r_ctrl.Scale.Y {type=MovieSceneScriptingFloatChannel, default=1, const [1,8]}
       pinky_01_r_ctrl.Scale.Z {type=MovieSceneScriptingFloatChannel, default=1, const [1,8]}
       pinky_02_r_ctrl.Location.X {type=MovieSceneScriptingFloatChannel, default=0, thinned_from=13, keys [[2162,0],[2759,0]]}
       pinky_02_r_ctrl.Location.Y {type=MovieSceneScriptingFloatChannel, default=0, thinned_from=13, keys [[2162,0],[2759,0]]}
       pinky_02_r_ctrl.Location.Z {type=MovieSceneScriptingFloatChannel, default=0, thinned_from=13, keys [[2162,0],[2759,0]]}
       pinky_02_r_ctrl.Rotation.X {
        type=MovieSceneScriptingFloatChannel
        default=0.3719
        thinned_from=18
        keys [[2162,0.3719],[2202,0],[2586,0],[2612,0.3719],[2627,0.325],[2644,0.3281],[2652,0.3719],[2664,0],[2674,0.2755],[2702,0.2755],[2709,0.186],[2720,0.3719],[2731,0.3719],[2739,0.1468],[2746,0],[2754,0],[2759,0.4415]]
       }
       pinky_02_r_ctrl.Rotation.Y {
        type=MovieSceneScriptingFloatChannel
        default=0.24
        thinned_from=18
        keys [[2162,0.24],[2202,0],[2586,0],[2612,0.24],[2627,0.2097],[2644,0.2118],[2652,0.24],[2664,0],[2674,0.1778],[2702,0.1778],[2709,0.12],[2720,0.24],[2731,0.24],[2739,0.0947],[2746,0],[2754,0],[2759,3.4359]]
       }
       pinky_02_r_ctrl.Rotation.Z {
        type=MovieSceneScriptingFloatChannel
        default=2.0416
        thinned_from=18
        keys [[2162,2.0416],[2202,-63.0936],[2586,-63.0936],[2612,2.0416],[2627,-6.1785],[2644,-5.6284],[2652,2.0416],[2664,-63.0936],[2674,-14.8453],[2702,-14.8453],[2709,-30.526],[2720,2.0416],[2731,2.0416],[2739,18.8891],[2746,-63.0936],[2754,-63.0936],[2759,26.1087]]
       }
       pinky_02_r_ctrl.Scale.X {type=MovieSceneScriptingFloatChannel, default=1, const [1,8]}
       pinky_02_r_ctrl.Scale.Y {type=MovieSceneScriptingFloatChannel, default=1, const [1,8]}
       pinky_02_r_ctrl.Scale.Z {type=MovieSceneScriptingFloatChannel, default=1, const [1,8]}
       pinky_03_r_ctrl.Location.X {type=MovieSceneScriptingFloatChannel, default=0.0002, thinned_from=13, keys [[2162,0.0002],[2731,0.0002],[2739,0.0001],[2759,0.0002]]}
       pinky_03_r_ctrl.Location.Y {type=MovieSceneScriptingFloatChannel, default=-0.0002, thinned_from=13, keys [[2162,-0.0002],[2731,-0.0002],[2739,-0.0001],[2759,-0.0002]]}
       pinky_03_r_ctrl.Location.Z {type=MovieSceneScriptingFloatChannel, default=0, thinned_from=13, keys [[2162,0],[2759,0]]}
       pinky_03_r_ctrl.Rotation.X {
        type=MovieSceneScriptingFloatChannel
        default=-0.2816
        thinned_from=18
        keys [[2162,-0.2816],[2202,0],[2586,0],[2612,-0.2816],[2627,-0.2461],[2644,-0.2484],[2652,-0.2816],[2664,0],[2674,-0.2086],[2702,-0.2086],[2709,-0.1408],[2720,-0.2816],[2731,-0.2816],[2739,-0.1112],[2746,0],[2754,0],[2759,0.1505]]
       }
       pinky_03_r_ctrl.Rotation.Y {
        type=MovieSceneScriptingFloatChannel
        default=1.0905
        thinned_from=18
        keys [[2162,1.0905],[2202,0],[2586,0],[2612,1.0905],[2627,0.9529],[2644,0.9621],[2652,1.0905],[2664,0],[2674,0.8078],[2702,0.8078],[2709,0.5453],[2720,1.0905],[2731,1.0905],[2739,0.4305],[2746,0],[2754,0],[2759,1.1162]]
       }
       pinky_03_r_ctrl.Rotation.Z {
        type=MovieSceneScriptingFloatChannel
        default=-5.0202
        thinned_from=18
        keys [[2162,-5.0202],[2202,-117.2954],[2586,-117.2954],[2612,-5.0202],[2627,-19.1894],[2644,-18.2412],[2652,-5.0202],[2664,-117.2954],[2674,-34.1286],[2702,-34.1286],[2709,-61.1578],[2720,-5.0202],[2731,-5.0202],[2739,5.5],[2746,-117.2954],[2754,-117.2954],[2759,17.1374]]
       }
       pinky_03_r_ctrl.Scale.X {type=MovieSceneScriptingFloatChannel, default=1, const [1,8]}
       pinky_03_r_ctrl.Scale.Y {type=MovieSceneScriptingFloatChannel, default=1, const [1,8]}
       pinky_03_r_ctrl.Scale.Z {type=MovieSceneScriptingFloatChannel, default=1, const [1,8]}
       hand_r_ik_ctrl.Location.X {type=MovieSceneScriptingFloatChannel, default=0, const [0,1021]}
       hand_r_ik_ctrl.Location.Y {type=MovieSceneScriptingFloatChannel, default=0, const [0,1021]}
       hand_r_ik_ctrl.Location.Z {type=MovieSceneScriptingFloatChannel, default=0, const [0,1021]}
       hand_r_ik_ctrl.Rotation.X {type=MovieSceneScriptingFloatChannel, default=0, const [0,1021]}
       hand_r_ik_ctrl.Rotation.Y {type=MovieSceneScriptingFloatChannel, default=0, const [0,1021]}
       hand_r_ik_ctrl.Rotation.Z {type=MovieSceneScriptingFloatChannel, default=0, const [0,1021]}
       hand_r_ik_ctrl.Scale.X {type=MovieSceneScriptingFloatChannel, default=1, const [1,1021]}
       hand_r_ik_ctrl.Scale.Y {type=MovieSceneScriptingFloatChannel, default=1, const [1,1021]}
       hand_r_ik_ctrl.Scale.Z {type=MovieSceneScriptingFloatChannel, default=1, const [1,1021]}
       Softness_2 {type=MovieSceneScriptingFloatChannel, default=0, const [0,1021]}
       hand_r_ikorient_ctrl.Rotation.Roll {type=MovieSceneScriptingFloatChannel, default=0, const [0,1021]}
       hand_r_ikorient_ctrl.Rotation.Pitch {type=MovieSceneScriptingFloatChannel, default=0, const [0,1021]}
       hand_r_ikorient_ctrl.Rotation.Yaw {type=MovieSceneScriptingFloatChannel, default=0, const [0,1021]}
       arm_r_pv_ik_ctrl.Location.X {type=MovieSceneScriptingFloatChannel, default=0, const [0,1021]}
       arm_r_pv_ik_ctrl.Location.Y {type=MovieSceneScriptingFloatChannel, default=0, const [0,1021]}
       arm_r_pv_ik_ctrl.Location.Z {type=MovieSceneScriptingFloatChannel, default=0, const [0,1021]}
       arm_r_pv_ik_ctrl.Rotation.X {type=MovieSceneScriptingFloatChannel, default=0, const [0,1021]}
       arm_r_pv_ik_ctrl.Rotation.Y {type=MovieSceneScriptingFloatChannel, default=0, const [0,1021]}
       arm_r_pv_ik_ctrl.Rotation.Z {type=MovieSceneScriptingFloatChannel, default=0, const [0,1021]}
       arm_r_pv_ik_ctrl.Scale.X {type=MovieSceneScriptingFloatChannel, default=1, const [1,1021]}
       arm_r_pv_ik_ctrl.Scale.Y {type=MovieSceneScriptingFloatChannel, default=1, const [1,1021]}
       arm_r_pv_ik_ctrl.Scale.Z {type=MovieSceneScriptingFloatChannel, default=1, const [1,1021]}
       body_ctrl.Location.X {
        type=MovieSceneScriptingFloatChannel
        default=0
        thinned_from=17
        keys [[2162,0],[2601,0],[2616,-5.011],[2635,1.3398],[2644,1.1866],[2663,-2.6742],[2708,1.3963],[2742,0],[2759,0]]
       }
       body_ctrl.Location.Y {
        type=MovieSceneScriptingFloatChannel
        default=1.2382
        thinned_from=18
        keys [[2162,1.2382],[2601,1.2382],[2614,1.2531],[2633,1.2382],[2759,1.2382]]
       }
       body_ctrl.Location.Z {
        type=MovieSceneScriptingFloatChannel
        default=-103.5107
        thinned_from=18
        keys [[2162,-103.5107],[2594,-103.5107],[2605,-105.65],[2617,-100.8531],[2638,-103.8665],[2649,-104.5975],[2675,-102.4382],[2703,-104.6645],[2733,-103.5107],[2747,-107.774],[2754,-107.9253],[2759,-103.5107]]
       }
       body_ctrl.Rotation.X {
        type=MovieSceneScriptingFloatChannel
        default=0
        thinned_from=21
        keys [[2162,0],[2601,0],[2615,-1.1655],[2627,-0.1942],[2638,0],[2733,0],[2747,3.8649],[2754,3.8649],[2759,0]]
       }
       body_ctrl.Rotation.Y {
        type=MovieSceneScriptingFloatChannel
        default=0
        thinned_from=21
        keys [[2162,0],[2601,0],[2615,-1.848],[2627,-0.308],[2638,0],[2733,0],[2747,3.411],[2754,3.411],[2759,0]]
       }
       body_ctrl.Rotation.Z {
        type=MovieSceneScriptingFloatChannel
        default=0
        thinned_from=41
        keys [[2162,0],[2202,0],[2208,6.4106],[2216,0],[2223,0],[2237,27.3981],[2241,27.3702],[2259,-59.5963],[2264,-65.4939],[2277,14.7023],[2286,0],[2293,0],[2307,-36.297],[2315,-46.8181],[2333,0],[2348,-12.7585],[2365,35.335],[2377,49.4849],[2412,-52.4063],[2424,-65.9287],[2441,-64.3003],[2456,-66.6569],[2473,-37.5404],[2483,-34.371],[2505,-54.3207],[2531,-62.086],[2561,-22.4791],[2582,-4.0422],[2601,23.0761],[2615,32.0595],[2627,32.3591],[2646,32.419],[2671,-10.516],[2702,-10.2927],[2717,0],[2759,0]]
       }
       body_ctrl.Scale.X {type=MovieSceneScriptingFloatChannel, default=1}
       body_ctrl.Scale.Y {type=MovieSceneScriptingFloatChannel, default=1}
       body_ctrl.Scale.Z {type=MovieSceneScriptingFloatChannel, default=1}
       hips_ctrl.Location.X {type=MovieSceneScriptingFloatChannel, default=0, const [0,15]}
       hips_ctrl.Location.Y {type=MovieSceneScriptingFloatChannel, default=0, const [0,15]}
       hips_ctrl.Location.Z {type=MovieSceneScriptingFloatChannel, default=0, const [0,15]}
       hips_ctrl.Rotation.X {type=MovieSceneScriptingFloatChannel, default=0, const [0,15]}
       hips_ctrl.Rotation.Y {type=MovieSceneScriptingFloatChannel, default=0, const [0,15]}
       hips_ctrl.Rotation.Z {type=MovieSceneScriptingFloatChannel, default=0, const [0,15]}
       hips_ctrl.Scale.X {type=MovieSceneScriptingFloatChannel, default=1, const [1,2]}
       hips_ctrl.Scale.Y {type=MovieSceneScriptingFloatChannel, default=1, const [1,2]}
       hips_ctrl.Scale.Z {type=MovieSceneScriptingFloatChannel, default=1, const [1,2]}
       hips_tan_ctrl.Location.X {type=MovieSceneScriptingFloatChannel, default=0, const [0,1021]}
       hips_tan_ctrl.Location.Y {type=MovieSceneScriptingFloatChannel, default=0.008, const [0.008,1021]}
       hips_tan_ctrl.Location.Z {type=MovieSceneScriptingFloatChannel, default=-0.0073, const [-0.0073,1021]}
       hips_tan_ctrl.Rotation.X {type=MovieSceneScriptingFloatChannel, default=0, const [0,1021]}
       hips_tan_ctrl.Rotation.Y {type=MovieSceneScriptingFloatChannel, default=0, const [0,1021]}
       hips_tan_ctrl.Rotation.Z {type=MovieSceneScriptingFloatChannel, default=0, const [0,1021]}
       hips_tan_ctrl.Scale.X {type=MovieSceneScriptingFloatChannel, default=1, const [1,1021]}
       hips_tan_ctrl.Scale.Y {type=MovieSceneScriptingFloatChannel, default=1, const [1,1021]}
       hips_tan_ctrl.Scale.Z {type=MovieSceneScriptingFloatChannel, default=1, const [1,1021]}
       spine_01_ctrl.Location.X {type=MovieSceneScriptingFloatChannel, default=0, const [0,15]}
       spine_01_ctrl.Location.Y {type=MovieSceneScriptingFloatChannel, default=0, const [0,15]}
       spine_01_ctrl.Location.Z {type=MovieSceneScriptingFloatChannel, default=0, const [0,15]}
       spine_01_ctrl.Rotation.X {
        type=MovieSceneScriptingFloatChannel
        default=-1.7101
        thinned_from=19
        keys [[2162,-1.7101],[2180,-1.7101],[2202,2.3786],[2208,5.7357],[2216,-6.6174],[2584,-6.5771],[2597,0.0043],[2606,1.6699],[2624,-0.6396],[2638,-1.7571],[2657,-0.2105],[2682,-5.9607],[2703,-0.2903],[2725,-1.7101],[2732,-1.7101],[2746,6.8166],[2753,8.3008],[2759,-4.4743]]
       }
       spine_01_ctrl.Rotation.Y {
        type=MovieSceneScriptingFloatChannel
        default=0
        thinned_from=21
        keys [[2162,0],[2180,0],[2202,0.6127],[2208,-0.0065],[2216,0],[2584,0],[2591,2.0949],[2597,1.5426],[2606,0.6227],[2624,0.1197],[2644,0.1197],[2661,0.2699],[2679,-0.1647],[2699,-0.1647],[2714,0],[2732,0],[2746,-0.9156],[2753,-1.097],[2759,-4.5037]]
       }
       spine_01_ctrl.Rotation.Z {
        type=MovieSceneScriptingFloatChannel
        default=0
        thinned_from=46
        keys [[2162,0],[2180,0],[2202,1.2735],[2208,3.7529],[2216,-0.4147],[2219,-1.4931],[2229,1.6572],[2236,1.6563],[2250,-0.8778],[2257,-3.8178],[2277,0],[2284,3.1409],[2296,-6.4824],[2303,-6.4767],[2317,-4.2485],[2331,-2.6834],[2342,-6.152],[2348,-6.5039],[2365,3.8175],[2372,5.6532],[2378,6.3554],[2392,1.2541],[2403,-2.9641],[2422,-4.594],[2427,-4.5875],[2446,-3.0673],[2460,-4.7656],[2479,-3.1517],[2485,-2.9316],[2505,-4.8879],[2521,-5.6469],[2556,0],[2584,0],[2597,6.9333],[2606,5.4734],[2624,4.0135],[2644,4.0135],[2665,-5.5298],[2699,-5.5252],[2714,0],[2732,0],[2746,-9.1898],[2753,-9.1898],[2759,0.1347]]
       }
       spine_01_ctrl.Scale.X {type=MovieSceneScriptingFloatChannel, default=1, const [1,1]}
       spine_01_ctrl.Scale.Y {type=MovieSceneScriptingFloatChannel, default=1, const [1,1]}
       spine_01_ctrl.Scale.Z {type=MovieSceneScriptingFloatChannel, default=1, const [1,1]}
       spine_02_ctrl.Location.X {type=MovieSceneScriptingFloatChannel, default=0, const [0,15]}
       spine_02_ctrl.Location.Y {type=MovieSceneScriptingFloatChannel, default=0, const [0,15]}
       spine_02_ctrl.Location.Z {type=MovieSceneScriptingFloatChannel, default=0, const [0,15]}
       spine_02_ctrl.Rotation.X {
        type=MovieSceneScriptingFloatChannel
        default=-0.6447
        thinned_from=20
        keys [[2162,-0.6447],[2180,-0.6447],[2202,2.3786],[2208,6.2676],[2216,-5.552],[2584,-5.5116],[2597,0.0016],[2606,2.279],[2624,0.4232],[2644,0.4232],[2665,-0.6417],[2699,-0.6417],[2714,-0.6447],[2732,-0.6447],[2746,7.8685],[2753,9.3512],[2759,-3.4056]]
       }
       spine_02_ctrl.Rotation.Y {
        type=MovieSceneScriptingFloatChannel
        default=0
        thinned_from=21
        keys [[2162,0],[2180,0],[2202,0.6127],[2208,-0.0355],[2216,0],[2584,0],[2591,2.0714],[2597,1.5122],[2606,0.5651],[2624,0.0451],[2644,0.0451],[2661,0.3726],[2679,-0.0621],[2699,-0.0621],[2714,0],[2732,0],[2746,-0.7453],[2753,-0.9261],[2759,-4.5054]]
       }
       spine_02_ctrl.Rotation.Z {
        type=MovieSceneScriptingFloatChannel
        default=0
        thinned_from=46
        keys [[2162,0],[2180,0],[2202,1.2735],[2208,3.7498],[2216,-0.4147],[2219,-1.4931],[2229,1.6572],[2236,1.6563],[2250,-0.8778],[2257,-3.8178],[2277,0],[2284,3.1409],[2296,-6.4824],[2303,-6.4767],[2317,-4.2485],[2331,-2.6834],[2342,-6.152],[2348,-6.5039],[2365,3.8175],[2372,5.6532],[2378,6.3554],[2392,1.2541],[2403,-2.9641],[2422,-4.594],[2427,-4.5875],[2446,-3.0673],[2460,-4.7656],[2479,-3.1517],[2485,-2.9316],[2505,-4.8879],[2521,-5.6469],[2556,0],[2584,0],[2597,6.9338],[2606,5.4745],[2624,4.0151],[2644,4.0151],[2665,-5.5319],[2699,-5.5273],[2714,0],[2732,0],[2746,-9.205],[2753,-9.205],[2759,0.0508]]
       }
       spine_02_ctrl.Scale.X {type=MovieSceneScriptingFloatChannel, default=1, const [1,1]}
       spine_02_ctrl.Scale.Y {type=MovieSceneScriptingFloatChannel, default=1, const [1,1]}
       spine_02_ctrl.Scale.Z {type=MovieSceneScriptingFloatChannel, default=1, const [1,1]}
       spine_03_ctrl.Location.X {type=MovieSceneScriptingFloatChannel, default=0, const [0,15]}
       spine_03_ctrl.Location.Y {type=MovieSceneScriptingFloatChannel, default=0, const [0,15]}
       spine_03_ctrl.Location.Z {type=MovieSceneScriptingFloatChannel, default=0, const [0,15]}
       spine_03_ctrl.Rotation.X {
        type=MovieSceneScriptingFloatChannel
        default=-0.6447
        thinned_from=19
        keys [[2162,-0.6447],[2180,-0.6447],[2202,2.3786],[2208,6.2676],[2216,-5.552],[2584,-5.5116],[2597,0.0016],[2606,2.279],[2624,-1.2562],[2638,0.9304],[2657,-2.7883],[2682,0.9422],[2703,-3.0992],[2725,-0.6447],[2732,-0.6447],[2746,7.8685],[2753,9.3512],[2759,-3.4056]]
       }
       spine_03_ctrl.Rotation.Y {
        type=MovieSceneScriptingFloatChannel
        default=0
        thinned_from=21
        keys [[2162,0],[2180,0],[2202,0.6127],[2208,-0.0355],[2216,0],[2584,0],[2591,2.0714],[2597,1.5122],[2606,0.5651],[2624,0.0451],[2644,0.0451],[2661,0.3726],[2679,-0.0621],[2699,-0.0621],[2714,0],[2732,0],[2746,-0.7453],[2753,-0.9261],[2759,-4.5054]]
       }
       spine_03_ctrl.Rotation.Z {
        type=MovieSceneScriptingFloatChannel
        default=0
        thinned_from=46
        keys [[2162,0],[2180,0],[2202,1.2735],[2208,3.7498],[2216,-0.4147],[2219,-1.4931],[2229,1.6572],[2236,1.6563],[2250,-0.8778],[2257,-3.8178],[2277,0],[2284,3.1409],[2296,-6.4824],[2303,-6.4767],[2317,-4.2485],[2331,-2.6834],[2342,-6.152],[2348,-6.5039],[2365,3.8175],[2372,5.6532],[2378,6.3554],[2392,1.2541],[2403,-2.9641],[2422,-4.594],[2427,-4.5875],[2446,-3.0673],[2460,-4.7656],[2479,-3.1517],[2485,-2.9316],[2505,-4.8879],[2521,-5.6469],[2556,0],[2584,0],[2597,6.9338],[2606,5.4745],[2624,4.0151],[2644,4.0151],[2665,-5.5319],[2699,-5.5273],[2714,0],[2732,0],[2746,-9.205],[2753,-9.205],[2759,0.0508]]
       }
       spine_03_ctrl.Scale.X {type=MovieSceneScriptingFloatChannel, default=1, const [1,1]}
       spine_03_ctrl.Scale.Y {type=MovieSceneScriptingFloatChannel, default=1, const [1,1]}
       spine_03_ctrl.Scale.Z {type=MovieSceneScriptingFloatChannel, default=1, const [1,1]}
       spine_04_ctrl.Location.X {type=MovieSceneScriptingFloatChannel, default=0, const [0,15]}
       spine_04_ctrl.Location.Y {type=MovieSceneScriptingFloatChannel, default=0, const [0,15]}
       spine_04_ctrl.Location.Z {type=MovieSceneScriptingFloatChannel, default=0, const [0,15]}
       spine_04_ctrl.Rotation.X {
        type=MovieSceneScriptingFloatChannel
        default=-0.6447
        thinned_from=19
        keys [[2162,-0.6447],[2180,-0.6447],[2202,2.3786],[2208,6.2676],[2216,-5.552],[2584,-5.5116],[2597,0.0016],[2606,2.279],[2624,-1.2562],[2638,0.9304],[2657,-2.7883],[2682,0.9422],[2703,-3.0992],[2725,-0.6447],[2732,-0.6447],[2746,7.8685],[2753,9.3512],[2759,-3.4056]]
       }
       spine_04_ctrl.Rotation.Y {
        type=MovieSceneScriptingFloatChannel
        default=0
        thinned_from=21
        keys [[2162,0],[2180,0],[2202,0.6127],[2208,-0.0355],[2216,0],[2584,0],[2591,2.0714],[2597,1.5122],[2606,0.5651],[2624,0.0451],[2644,0.0451],[2661,0.3726],[2679,-0.0621],[2699,-0.0621],[2714,0],[2732,0],[2746,-0.7453],[2753,-0.9261],[2759,-4.5054]]
       }
       spine_04_ctrl.Rotation.Z {
        type=MovieSceneScriptingFloatChannel
        default=0
        thinned_from=46
        keys [[2162,0],[2180,0],[2202,1.2735],[2208,3.7498],[2216,-0.4147],[2219,-1.4931],[2229,1.6572],[2236,1.6563],[2250,-0.8778],[2257,-3.8178],[2277,0],[2284,3.1409],[2296,-6.4824],[2303,-6.4767],[2317,-4.2485],[2331,-2.6834],[2342,-6.152],[2348,-6.5039],[2365,3.8175],[2372,5.6532],[2378,6.3554],[2392,1.2541],[2403,-2.9641],[2422,-4.594],[2427,-4.5875],[2446,-3.0673],[2460,-4.7656],[2479,-3.1517],[2485,-2.9316],[2505,-4.8879],[2521,-5.6469],[2556,0],[2584,0],[2597,6.9338],[2606,5.4745],[2624,4.0151],[2644,4.0151],[2665,-5.5319],[2699,-5.5273],[2714,0],[2732,0],[2746,-9.205],[2753,-9.205],[2759,0.0508]]
       }
       spine_04_ctrl.Scale.X {type=MovieSceneScriptingFloatChannel, default=1, const [1,1]}
       spine_04_ctrl.Scale.Y {type=MovieSceneScriptingFloatChannel, default=1, const [1,1]}
       spine_04_ctrl.Scale.Z {type=MovieSceneScriptingFloatChannel, default=1, const [1,1]}
       head_ctrl.Location.X {type=MovieSceneScriptingFloatChannel, default=0, thinned_from=21, keys [[2162,0],[2759,0]]}
       head_ctrl.Location.Y {
        type=MovieSceneScriptingFloatChannel
        default=0
        thinned_from=21
        keys [[2162,0],[2268,0],[2425,-0.0043],[2521,-0.0032],[2534,-0.0029],[2574,-0.0002],[2586,0],[2759,0]]
       }
       head_ctrl.Location.Z {
        type=MovieSceneScriptingFloatChannel
        default=0
        thinned_from=21
        keys [[2162,0],[2268,0],[2425,-0.4436],[2521,-0.3336],[2534,-0.294],[2574,-0.0163],[2586,0],[2759,0]]
       }
       head_ctrl.Rotation.X {
        type=MovieSceneScriptingFloatChannel
        default=6.7767
        thinned_from=51
        keys [[2162,6.7767],[2180,6.7767],[2202,24.7706],[2208,16.6335],[2216,-49.2702],[2242,-49.2664],[2251,-71.8041],[2260,-66.8957],[2268,-57.4207],[2276,-52.8001],[2288,-48.1701],[2304,-53.5231],[2314,-54.722],[2326,-55.6236],[2328,-55.623],[2336,-48.1701],[2352,-48.8433],[2362,-58.0644],[2377,-58.0035],[2389,-57.7453],[2401,-62.0249],[2412,-62.5412],[2425,-57.5668],[2441,-56.1872],[2459,-72.8942],[2477,-71.4105],[2499,-68.3895],[2503,-67.7295],[2521,-67.4077],[2534,-66.2151],[2560,-58.4918],[2574,-51.8833],[2586,-48.8091],[2612,5.3632],[2644,5.3632],[2651,6.0573],[2658,2.4048],[2697,2.4048],[2704,5.9347],[2709,3.0302],[2718,2.1458],[2722,-2.4567],[2727,12.4573],[2730,11.632],[2733,9.1492],[2745,36.0075],[2754,36.0075],[2759,-46.7412]]
       }
       head_ctrl.Rotation.Y {
        type=MovieSceneScriptingFloatChannel
        default=0
        thinned_from=49
        keys [[2162,0],[2180,0],[2202,-1.901],[2208,-0.9505],[2216,0],[2242,0],[2251,32.8045],[2260,37.9045],[2268,14.2604],[2276,26.3604],[2288,25.7149],[2304,0],[2314,1.493],[2326,1.8848],[2328,1.8845],[2336,25.7149],[2352,9.9251],[2362,5.1494],[2377,4.6329],[2389,4.8615],[2401,27.6094],[2412,22.1579],[2425,0],[2441,1.2089],[2459,40.5617],[2477,34.7785],[2499,26.6238],[2503,24.5068],[2521,29.6779],[2534,26.7748],[2560,9.1502],[2574,1.6481],[2586,0],[2612,4.5926],[2642,4.5926],[2655,-1.6223],[2668,-1.089],[2697,-1.089],[2704,-2.353],[2710,-0.4578],[2718,0],[2721,0],[2728,-2.0623],[2734,-1.055],[2745,-0.6104],[2754,-0.6104],[2759,0.1435]]
       }
       head_ctrl.Rotation.Z {
        type=MovieSceneScriptingFloatChannel
        default=0
        thinned_from=49
        keys [[2162,0],[2180,0],[2202,-14.7722],[2208,-7.3861],[2216,0],[2242,0],[2251,13.0495],[2260,19.888],[2268,7.4822],[2276,6.3607],[2288,18.9362],[2304,0],[2314,11.1019],[2326,2.1699],[2328,2.1448],[2336,18.9362],[2352,7.3087],[2362,-10.6363],[2377,-10.5663],[2389,-10.3887],[2401,10.9933],[2412,15.0592],[2425,0],[2441,5.7675],[2459,11.9134],[2477,12.6403],[2499,11.3033],[2503,10.5432],[2521,14.15],[2534,12.7659],[2560,4.3627],[2574,0.7858],[2586,0],[2612,72.6075],[2642,72.6075],[2659,-54.1256],[2681,-54.1738],[2689,-40.3916],[2697,-40.3916],[2704,-12.3842],[2710,2.6846],[2718,0],[2720,0],[2728,0.7708],[2734,0.2558],[2745,-3.5452],[2754,-3.5452],[2759,5.2947]]
       }
       head_ctrl.Scale.X {type=MovieSceneScriptingFloatChannel, default=1, const [1,8]}
       head_ctrl.Scale.Y {type=MovieSceneScriptingFloatChannel, default=1, const [1,8]}
       head_ctrl.Scale.Z {type=MovieSceneScriptingFloatChannel, default=1, const [1,8]}
       clavicle_l_ctrl.Location.X {type=MovieSceneScriptingFloatChannel, default=0, const [0,6]}
       clavicle_l_ctrl.Location.Y {type=MovieSceneScriptingFloatChannel, default=0, const [0,6]}
       clavicle_l_ctrl.Location.Z {type=MovieSceneScriptingFloatChannel, default=0, const [0,6]}
       clavicle_l_ctrl.Rotation.X {
        type=MovieSceneScriptingFloatChannel
        default=-0.6816
        thinned_from=18
        keys [[2162,-0.6816],[2180,-0.6816],[2202,-5.9146],[2208,-2.923],[2211,-2.923],[2220,0.0687],[2227,0.0687],[2586,0.0625],[2612,-0.6816],[2734,-0.6816],[2749,-7.3766],[2754,-7.3766],[2759,-5.5968]]
       }
       clavicle_l_ctrl.Rotation.Y {
        type=MovieSceneScriptingFloatChannel
        default=-0.6185
        thinned_from=18
        keys [[2162,-0.6185],[2180,-0.6185],[2202,7.2579],[2208,4.9205],[2211,4.9205],[2220,2.5827],[2586,2.5555],[2612,-0.6185],[2734,-0.6185],[2749,4.8502],[2754,4.8502],[2759,-3.8781]]
       }
       clavicle_l_ctrl.Rotation.Z {
        type=MovieSceneScriptingFloatChannel
        default=-5.6369
        thinned_from=18
        keys [[2162,-5.6369],[2180,-5.6369],[2202,-10.9571],[2208,-8.8071],[2211,-8.8071],[2220,-6.657],[2586,-6.6483],[2612,-5.6369],[2734,-5.6369],[2749,-5.1362],[2754,-5.1362],[2759,-7.5744]]
       }
       clavicle_l_ctrl.Scale.X {type=MovieSceneScriptingFloatChannel, default=1, const [1,3]}
       clavicle_l_ctrl.Scale.Y {type=MovieSceneScriptingFloatChannel, default=1, const [1,3]}
       clavicle_l_ctrl.Scale.Z {type=MovieSceneScriptingFloatChannel, default=1, const [1,3]}
       clavicle_r_ctrl.Location.X {type=MovieSceneScriptingFloatChannel, default=0, const [0,16]}
       clavicle_r_ctrl.Location.Y {type=MovieSceneScriptingFloatChannel, default=0, const [0,16]}
       clavicle_r_ctrl.Location.Z {type=MovieSceneScriptingFloatChannel, default=0, const [0,16]}
       clavicle_r_ctrl.Rotation.X {
        type=MovieSceneScriptingFloatChannel
        default=-0.6816
        thinned_from=18
        keys [[2162,-0.6816],[2180,-0.6816],[2202,-5.4053],[2208,-3.8937],[2222,-2.3818],[2586,-2.3674],[2612,-0.6816],[2656,-0.6816],[2674,-3.6093],[2706,-3.6093],[2723,-0.6816],[2733,-0.6816],[2744,0.1603],[2748,0.4146],[2750,0.5123],[2754,0.5123],[2759,-17.0181]]
       }
       clavicle_r_ctrl.Rotation.Y {
        type=MovieSceneScriptingFloatChannel
        default=-0.6185
        thinned_from=18
        keys [[2162,-0.6185],[2180,-0.6185],[2202,-0.8804],[2208,-1.3462],[2222,-1.7583],[2586,-1.749],[2612,-0.6185],[2656,-0.6185],[2674,0.3391],[2706,0.3391],[2723,-0.6185],[2733,-0.6185],[2744,8.5103],[2748,12.3705],[2750,13.8526],[2754,13.8526],[2759,39.4367]]
       }
       clavicle_r_ctrl.Rotation.Z {
        type=MovieSceneScriptingFloatChannel
        default=-5.6369
        thinned_from=18
        keys [[2162,-5.6369],[2180,-5.6369],[2202,-15.4541],[2208,-11.9366],[2222,-8.4189],[2586,-8.3952],[2612,-5.6369],[2656,-5.6369],[2674,1.1684],[2706,1.1684],[2723,-5.6369],[2733,-5.6369],[2744,10.3702],[2748,10.7242],[2750,10.8601],[2754,10.8601],[2759,-5.9081]]
       }
       clavicle_r_ctrl.Scale.X {type=MovieSceneScriptingFloatChannel, default=1, const [1,3]}
       clavicle_r_ctrl.Scale.Y {type=MovieSceneScriptingFloatChannel, default=1, const [1,3]}
       clavicle_r_ctrl.Scale.Z {type=MovieSceneScriptingFloatChannel, default=1, const [1,3]}
       neck_01_ctrl.Location.X {type=MovieSceneScriptingFloatChannel, default=0, thinned_from=20, keys [[2162,0],[2425,0.0002],[2759,0]]}
       neck_01_ctrl.Location.Y {type=MovieSceneScriptingFloatChannel, default=0, thinned_from=20, keys [[2162,0],[2268,0],[2425,-0.0003],[2759,0]]}
       neck_01_ctrl.Location.Z {type=MovieSceneScriptingFloatChannel, default=0, thinned_from=20, keys [[2162,0],[2759,0]]}
       neck_01_ctrl.Rotation.X {
        type=MovieSceneScriptingFloatChannel
        default=-3.8639
        thinned_from=42
        keys [[2162,-3.8639],[2202,-3.8639],[2208,-6.0378],[2216,-8.2117],[2242,-8.2112],[2251,-6.2044],[2260,-6.3047],[2268,-7.5064],[2276,-6.7052],[2288,-11.204],[2304,-13.3988],[2314,-13.3898],[2326,-8.2112],[2336,-8.2492],[2362,-8.2112],[2389,-8.2151],[2401,-6.5395],[2412,-6.6323],[2425,-8.2112],[2459,-7.5633],[2477,-7.5968],[2503,-7.7194],[2521,-6.3047],[2534,-6.4877],[2560,-7.599],[2574,-8.072],[2586,-8.1759],[2612,-1.2474],[2642,-1.2474],[2657,-3.5344],[2697,-3.5344],[2716,-3.8639],[2731,-3.8639],[2745,10.4503],[2754,10.4503],[2759,-13.9965]]
       }
       neck_01_ctrl.Rotation.Y {
        type=MovieSceneScriptingFloatChannel
        default=0
        thinned_from=42
        keys [[2162,0],[2242,0],[2251,13.7202],[2260,18.1235],[2268,6.8184],[2276,14.3552],[2288,15.048],[2304,0],[2328,0],[2336,14.7468],[2352,5.6917],[2362,0],[2377,0],[2389,1.7649],[2401,11.7277],[2412,14.4516],[2425,0],[2459,19.1265],[2477,18.0828],[2499,14.9512],[2503,14.2553],[2521,18.1235],[2534,16.3507],[2560,5.5878],[2574,1.0065],[2586,0],[2612,1.5984],[2642,1.5984],[2657,-1.0032],[2697,-1.0032],[2716,0],[2731,0],[2745,14.0474],[2754,14.0474],[2759,11.2678]]
       }
       neck_01_ctrl.Rotation.Z {
        type=MovieSceneScriptingFloatChannel
        default=0
        thinned_from=42
        keys [[2162,0],[2242,0],[2251,5.7888],[2260,6.4478],[2268,2.4258],[2276,5.1071],[2288,5.1411],[2304,0],[2328,0],[2336,3.9533],[2352,1.5258],[2362,0],[2377,0],[2389,-0.2548],[2401,4.7815],[2412,4.3955],[2425,0],[2459,6.7756],[2477,6.4059],[2499,5.2965],[2503,5.05],[2521,6.4478],[2534,5.817],[2560,1.988],[2574,0.3581],[2586,0],[2612,15.3361],[2642,15.3361],[2657,-15.8784],[2697,-15.8784],[2716,0],[2731,0],[2745,-1.8054],[2754,-1.8054],[2759,3.7407]]
       }
       neck_01_ctrl.Scale.X {type=MovieSceneScriptingFloatChannel, default=1, const [1,7]}
       neck_01_ctrl.Scale.Y {type=MovieSceneScriptingFloatChannel, default=1, const [1,7]}
       neck_01_ctrl.Scale.Z {type=MovieSceneScriptingFloatChannel, default=1, const [1,7]}
       neck_02_ctrl.Location.X {type=MovieSceneScriptingFloatChannel, default=0, const [0,20]}
       neck_02_ctrl.Location.Y {
        type=MovieSceneScriptingFloatChannel
        default=0
        thinned_from=20
        keys [[2162,0],[2268,0],[2425,0.1243],[2534,0.0486],[2574,0.0027],[2586,0],[2759,0]]
       }
       neck_02_ctrl.Location.Z {
        type=MovieSceneScriptingFloatChannel
        default=0
        thinned_from=20
        keys [[2162,0],[2268,0],[2425,0.3714],[2534,0.1453],[2574,0.008],[2586,0],[2759,0]]
       }
       neck_02_ctrl.Rotation.X {
        type=MovieSceneScriptingFloatChannel
        default=0.6608
        thinned_from=42
        keys [[2162,0.6608],[2202,0.6608],[2208,-3.7754],[2216,-8.2117],[2503,-8.1992],[2534,-8.2054],[2586,-8.1387],[2612,0.6608],[2759,0.6608]]
       }
       neck_02_ctrl.Rotation.Y {type=MovieSceneScriptingFloatChannel, default=0, const [0,42]}
       neck_02_ctrl.Rotation.Z {type=MovieSceneScriptingFloatChannel, default=0, const [0,42]}
       neck_02_ctrl.Scale.X {type=MovieSceneScriptingFloatChannel, default=1, const [1,7]}
       neck_02_ctrl.Scale.Y {type=MovieSceneScriptingFloatChannel, default=1, const [1,7]}
       neck_02_ctrl.Scale.Z {type=MovieSceneScriptingFloatChannel, default=1, const [1,7]}
       head_ik_ctrl.Location.X {type=MovieSceneScriptingFloatChannel, default=0, thinned_from=1021, keys [[1800,0],[2820,0]]}
       head_ik_ctrl.Location.Y {type=MovieSceneScriptingFloatChannel, default=0, thinned_from=1021, keys [[1800,0],[2424,0],[2425,-0.0297],[2426,0],[2820,0]]}
       head_ik_ctrl.Location.Z {type=MovieSceneScriptingFloatChannel, default=0, thinned_from=1021, keys [[1800,0],[2424,0],[2425,-0.1576],[2426,0],[2820,0]]}
       head_ik_ctrl.Rotation.X {type=MovieSceneScriptingFloatChannel, default=0, thinned_from=1021, keys [[1800,0],[2424,0],[2425,0.4736],[2426,0],[2820,0]]}
       head_ik_ctrl.Rotation.Y {type=MovieSceneScriptingFloatChannel, default=0, const [0,1021]}
       head_ik_ctrl.Rotation.Z {type=MovieSceneScriptingFloatChannel, default=0, const [0,1021]}
       head_ik_ctrl.Scale.X {type=MovieSceneScriptingFloatChannel, default=1, const [1,1021]}
       head_ik_ctrl.Scale.Y {type=MovieSceneScriptingFloatChannel, default=1, const [1,1021]}
       head_ik_ctrl.Scale.Z {type=MovieSceneScriptingFloatChannel, default=1, const [1,1021]}
       chest_ctrl.Location.X {type=MovieSceneScriptingFloatChannel, default=0, const [0,1021]}
       chest_ctrl.Location.Y {type=MovieSceneScriptingFloatChannel, default=0, const [0,1021]}
       chest_ctrl.Location.Z {type=MovieSceneScriptingFloatChannel, default=0, const [0,1021]}
       chest_ctrl.Rotation.X {type=MovieSceneScriptingFloatChannel, default=0, const [0,1021]}
       chest_ctrl.Rotation.Y {type=MovieSceneScriptingFloatChannel, default=0, const [0,1021]}
       chest_ctrl.Rotation.Z {type=MovieSceneScriptingFloatChannel, default=0, const [0,1021]}
       chest_ctrl.Scale.X {type=MovieSceneScriptingFloatChannel, default=1, const [1,1021]}
       chest_ctrl.Scale.Y {type=MovieSceneScriptingFloatChannel, default=1, const [1,1021]}
       chest_ctrl.Scale.Z {type=MovieSceneScriptingFloatChannel, default=1, const [1,1021]}
       chest_tan_ctrl.Location.X {type=MovieSceneScriptingFloatChannel, default=0, const [0,1021]}
       chest_tan_ctrl.Location.Y {type=MovieSceneScriptingFloatChannel, default=-0.0303, const [-0.0303,1021]}
       chest_tan_ctrl.Location.Z {type=MovieSceneScriptingFloatChannel, default=0.265, const [0.265,1021]}
       chest_tan_ctrl.Rotation.X {type=MovieSceneScriptingFloatChannel, default=0, const [0,1021]}
       chest_tan_ctrl.Rotation.Y {type=MovieSceneScriptingFloatChannel, default=0, const [0,1021]}
       chest_tan_ctrl.Rotation.Z {type=MovieSceneScriptingFloatChannel, default=0, const [0,1021]}
       chest_tan_ctrl.Scale.X {type=MovieSceneScriptingFloatChannel, default=1, const [1,1021]}
       chest_tan_ctrl.Scale.Y {type=MovieSceneScriptingFloatChannel, default=1, const [1,1021]}
       chest_tan_ctrl.Scale.Z {type=MovieSceneScriptingFloatChannel, default=1, const [1,1021]}
       foot_l_ik_ctrl.Space {
        type=MovieSceneScriptingSpaceChannel
        const [
         "<Struct 'MovieSceneControlRigSpaceBaseKey' (0x85256187590) {space_type: Parent, control_rig_element: {type: Null, name: \"leg_l_controls_space\"}}>"
         1
        ]
       }
       leg_l_pv_ik_ctrl.Space {
        type=MovieSceneScriptingSpaceChannel
        const [
         "<Struct 'MovieSceneControlRigSpaceBaseKey' (0x852563d3120) {space_type: Parent, control_rig_element: {type: Null, name: \"leg_l_pv_footFollow\"}}>"
         1
        ]
       }
       foot_r_ik_ctrl.Space {
        type=MovieSceneScriptingSpaceChannel
        const [
         "<Struct 'MovieSceneControlRigSpaceBaseKey' (0x852563d3320) {space_type: Parent, control_rig_element: {type: Null, name: \"leg_r_controls_space\"}}>"
         1
        ]
       }
       leg_r_pv_ik_ctrl.Space {
        type=MovieSceneScriptingSpaceChannel
        const [
         "<Struct 'MovieSceneControlRigSpaceBaseKey' (0x852563d3520) {space_type: Parent, control_rig_element: {type: Null, name: \"leg_r_pv_footFollow\"}}>"
         1
        ]
       }
       body_ctrl.Space {
        type=MovieSceneScriptingSpaceChannel
        const [
         "<Struct 'MovieSceneControlRigSpaceBaseKey' (0x852563d3740) {space_type: Parent, control_rig_element: {type: Control, name: \"body_offset_ctrl\"}}>"
         1
        ]
       }
      ]
     }]
   }
   {type=MovieSceneSpawnTrack, sections [{type=MovieSceneSpawnSection, channels [None {type=MovieSceneScriptingBoolChannel, default=true}]}]}
   {
    type=MovieScene3DTransformTrack
    sections [
     {
      type=MovieScene3DTransformSection
      end_frame=3770
      channels [
       Location.X {type=MovieSceneScriptingDoubleChannel, default=-52714, const [495.3745,1]}
       Location.Y {type=MovieSceneScriptingDoubleChannel, default=25000, const [-613.4579,1]}
       Location.Z {type=MovieSceneScriptingDoubleChannel, default=232776.6406, const [32880.7305,1]}
       Rotation.X {type=MovieSceneScriptingDoubleChannel, default=0, keys [[1800,184.4053],[1822,178.8983]]}
       Rotation.Y {type=MovieSceneScriptingDoubleChannel, default=0, keys [[1800,13.4808],[1822,0.7977]]}
       Rotation.Z {type=MovieSceneScriptingDoubleChannel, default=0, keys [[1800,-14.1991],[1822,51.3411]]}
       Scale.X {type=MovieSceneScriptingDoubleChannel, default=4, const [4,1]}
       Scale.Y {type=MovieSceneScriptingDoubleChannel, default=4, const [4,1]}
       Scale.Z {type=MovieSceneScriptingDoubleChannel, default=4, const [4,1]}
      ]
     }
     {
      type=MovieScene3DTransformSection
      channels [
       Location.X {type=MovieSceneScriptingDoubleChannel, default=0, const [14436,1]}
       Location.Y {type=MovieSceneScriptingDoubleChannel, default=0, const [76088,1]}
       Location.Z {type=MovieSceneScriptingDoubleChannel, default=0, const [3072,1]}
       Rotation.X {type=MovieSceneScriptingDoubleChannel, default=0, const [0,1]}
       Rotation.Y {type=MovieSceneScriptingDoubleChannel, default=0, const [0,1]}
       Rotation.Z {type=MovieSceneScriptingDoubleChannel, default=0, const [0,1]}
       Scale.X {type=MovieSceneScriptingDoubleChannel, default=0, const [0,1]}
       Scale.Y {type=MovieSceneScriptingDoubleChannel, default=0, const [0,1]}
       Scale.Z {type=MovieSceneScriptingDoubleChannel, default=0, const [0,1]}
      ]
     }
    ]
   }
  ]
 }
 Eye Laser Temp R {
  id=#2
  possessed_class=FortStaticMeshActor
  tracks [
   {type=MovieSceneSpawnTrack, sections [{type=MovieSceneSpawnSection, channels [None {type=MovieSceneScriptingBoolChannel, default=true}]}]}
   {
    type=MovieScene3DTransformTrack
    sections [{
      type=MovieScene3DTransformSection
      channels [
       Location.X {type=MovieSceneScriptingDoubleChannel, default=4.8795, const [4.88,1]}
       Location.Y {type=MovieSceneScriptingDoubleChannel, default=8.96, const [8.9632,1]}
       Location.Z {type=MovieSceneScriptingDoubleChannel, default=-2.7155, const [2.72,1]}
       Rotation.X {
        type=MovieSceneScriptingDoubleChannel
        default=90
        thinned_from=620
        keys [[1995.5,90],[2202,88.5242],[2203,67.6996],[2204,64.916],[2205,61.3553],[2206,53.2529],[2207,32.7349],[2208,3.2593],[2209,-6.1872],[2210,4.7803],[2211,17.6954],[2212,25.6968],[2213,29.0374],[2214,28.6525],[2215,24.8315],[2216,17.464],[2217,14.6501],[2218,16.3152],[2219,17.7654],[2220,19.0502],[2221,20.1925],[2222,21.208],[2223,22.1105],[2224,22.9123],[2225,24.022],[2226,25.7742],[2227,28.0572],[2228,30.7486],[2229,33.7205],[2231,39.9989],[2232,43.0674],[2233,45.9444],[2234,48.5325],[2235,50.7406],[2236,52.4804],[2237,53.6629],[2238,54.1938],[2241,54.5867],[2242,54.6359],[2243,53.7547],[2244,51.4239],[2245,47.7427],[2246,42.7877],[2247,36.6123],[2248,29.2676],[2249,20.8408],[2250,11.5067],[2251,1.5736],[2252,-10.1015],[2253,-24.0582],[2254,-37.9091],[2255,-49.1029],[2256,-56.2037],[2257,-58.4477],[2258,-54.4338],[2259,-41.2378],[2260,-18.5982],[2261,4.2417],[2262,20.3791],[2263,29.8374],[2264,35.0258],[2265,37.5869],[2266,36.6209],[2267,32.5076],[2268,27.4772],[2269,22.9608],[2270,20.311],[2271,20.1928],[2272,22.0928],[2273,25.3203],[2275,32.9576],[2276,36.108],[2277,38.1251],[2278,38.6467],[2279,38.0775],[2280,37.145],[2281,36.0821],[2282,35.092],[2283,34.3328],[2284,33.9105],[2285,34.0885],[2286,35.0912],[2287,36.89],[2288,39.2735],[2289,42.027],[2290,45.0851],[2291,48.3768],[2295,62.3614],[2296,65.6624],[2297,68.8118],[2298,71.7759],[2299,74.5099],[2300,76.9604],[2301,79.0702],[2302,80.7824],[2304,83.241],[2305,84.2384],[2306,85.0568],[2307,85.6782],[2309,86.3553],[2315,86.7725],[2318,87.4844],[2320,88.5303],[2321,89.3402],[2322,90.4109],[2323,91.8113],[2324,93.6284],[2325,95.9664],[2326,98.9293],[2327,102.5559],[2329,110.6375],[2330,113.5318],[2331,114.6711],[2332,114.1331],[2335,110.662],[2336,109.8556],[2337,109.4248],[2338,107.3074],[2339,102.2267],[2340,94.4787],[2341,83.1205],[2342,65.5086],[2343,43.5386],[2344,32.0502],[2345,32.8954],[2346,38.0463],[2347,43.6118],[2348,48.6256],[2349,53.005],[2350,56.9187],[2351,60.5018],[2352,63.8061],[2353,66.8787],[2354,69.755],[2355,72.458],[2356,74.9997],[2357,77.3831],[2358,79.6043],[2359,81.6555],[2360,83.5261],[2361,85.205],[2362,86.7025],[2363,88.0252],[2364,89.1585],[2365,90.0918],[2366,90.8173],[2368,91.7051],[2373,92.5022],[2377,92.2915],[2380,91.5401],[2383,90.1721],[2386,88.155],[2389,85.4782],[2392,82.3484],[2393,80.1818],[2394,74.7466],[2395,68.0707],[2396,62.7005],[2397,59.595],[2398,58.342],[2399,58.0995],[2400,58.0512],[2401,58.257],[2402,59.0878],[2403,60.3483],[2404,61.8909],[2409,70.6385],[2411,73.6642],[2413,76.1585],[2417,80.4679],[2421,84.0953],[2423,85.829],[2426,89.3687],[2427,90.3881],[2428,91.1049],[2439,93.9242],[2443,94.4466],[2444,94.9181],[2445,95.1787],[2446,94.6157],[2447,93.0378],[2448,90.312],[2449,86.458],[2450,81.7343],[2451,76.6258],[2452,71.6762],[2453,67.303],[2454,63.7332],[2455,61.0034],[2456,59.0187],[2457,57.6185],[2458,56.6237],[2461,54.4858],[2462,53.9203],[2463,53.4936],[2465,52.955],[2468,52.6523],[2473,52.5333],[2475,52.2925],[2482,50.5474],[2485,50.1724],[2487,50.1646],[2488,50.254],[2489,50.5034],[2490,50.9943],[2491,51.7093],[2492,52.6303],[2493,53.7384],[2494,55.014],[2495,56.4403],[2497,59.6717],[2499,63.3052],[2502,69.3051],[2505,75.8209],[2507,80.4656],[2509,85.4528],[2511,90.7943],[2522,123.5169],[2524,129.0407],[2525,131.5815],[2526,133.8544],[2527,135.7905],[2528,137.3122],[2529,138.333],[2530,138.7607],[2531,138.5039],[2532,137.4823],[2533,135.6526],[2534,132.9955],[2535,129.5513],[2536,125.4432],[2540,106.7181],[2541,102.4597],[2542,98.5598],[2543,95.0027],[2544,91.739],[2546,85.8419],[2548,80.4749],[2549,77.9663],[2550,75.6238],[2551,73.519],[2552,71.7371],[2553,70.3675],[2554,69.4717],[2555,69.0961],[2556,69.2528],[2559,70.8846],[2569,77.5285],[2576,80.977],[2584,83.6131],[2585,84.3901],[2586,86.0899],[2587,88.6381],[2588,91.9059],[2589,95.7396],[2590,99.794],[2591,104.646],[2592,110.3796],[2593,118.3085],[2594,131.7665],[2595,175.7576],[2596,-134.6419],[2597,-109.5108],[2598,-100.382],[2599,-95.9949],[2601,-89.9148],[2602,-85.9032],[2603,-83.4504],[2604,-91.2932],[2605,70.7047],[2606,32.0277],[2607,5.2947],[2608,-8.3035],[2609,-16.1366],[2610,-21.3788],[2611,-25.3067],[2612,-28.5061],[2613,-31.2712],[2614,-33.9515],[2615,-36.3987],[2616,-37.5898],[2617,-37.4636],[2618,-36.6981],[2619,-35.3822],[2620,-33.5904],[2621,-31.4221],[2622,-28.9882],[2624,-23.8494],[2625,-21.4473],[2626,-19.3715],[2627,-17.7832],[2628,-16.8436],[2630,-15.7349],[2631,-15.2079],[2632,-14.7116],[2633,-14.2551],[2634,-13.8476],[2635,-13.4979],[2636,-13.2148],[2637,-13.0055],[2638,-12.879],[2639,-12.8455],[2643,-12.8188],[2644,16.6708],[2645,72.9901],[2646,92.6793],[2647,138.0328],[2648,-141.5365],[2649,-137.7397],[2650,-145.0431],[2651,-156.008],[2652,-168.8942],[2653,177.2618],[2654,163.3231],[2655,150.1659],[2656,138.5242],[2657,128.9431],[2658,121.8618],[2659,117.6964],[2660,115.3077],[2661,113.2446],[2662,111.4622],[2664,108.5617],[2669,103.1321],[2670,102.212],[2671,101.6351],[2682,101.4513],[2683,102.1397],[2684,104.2216],[2686,110.2093],[2687,112.5461],[2688,114.0962],[2689,114.9411],[2690,115.2038],[2698,115.2931],[2699,116.4363],[2700,119.615],[2701,124.4928],[2702,130.7411],[2703,138.0352],[2704,146.1744],[2705,154.9088],[2706,162.4518],[2707,167.6226],[2708,170.9454],[2709,172.9598],[2712,176.0433],[2714,177.0945],[2717,177.3193],[2725,177.5378],[2729,178.3268],[2737,177.1154],[2745,173.3552],[2755,173.111],[2756,173.8689],[2757,176.3308],[2758,179.7602],[2759,-176.7598],[2760,-173.2696],[2761,-170.7564],[2762,-169.6407],[2763,-171.6141],[2764,-178.1499],[2765,170.4009],[2766,155.9688],[2767,142.0568],[2768,130.9002],[2769,122.6673],[2770,116.6818],[2771,112.2607],[2772,108.9116],[2773,106.3087],[2774,104.2383],[2775,102.5583],[2776,101.1719],[2778,99.028],[2780,97.4588],[2784,95.3473],[2793,92.9778],[2800,92.1772],[2820,92.1772]]
       }
       Rotation.Y {
        type=MovieSceneScriptingDoubleChannel
        default=0
        thinned_from=620
        keys [[1995.5,0],[2202,3.3482],[2203,-2.1472],[2204,-1.5038],[2205,-0.4021],[2206,2.4034],[2207,9.6964],[2208,24.0833],[2209,37.971],[2210,46.1711],[2211,47.8486],[2212,47.2178],[2213,46.6959],[2214,46.9768],[2215,47.9814],[2216,49.1307],[2217,48.8593],[2218,47.8141],[2219,46.9059],[2221,45.3802],[2223,44.1349],[2224,43.5902],[2225,42.957],[2226,42.1174],[2227,41.0897],[2228,39.8915],[2229,38.5436],[2230,37.0735],[2233,32.3493],[2234,30.8567],[2235,29.5173],[2236,28.4044],[2237,27.5918],[2238,27.1507],[2241,26.534],[2242,26.3999],[2243,26.9134],[2244,28.3238],[2245,30.4594],[2246,33.1148],[2248,39.032],[2249,41.7736],[2250,44.0279],[2251,45.5819],[2252,45.6475],[2253,42.564],[2254,34.861],[2255,21.6471],[2256,3.1729],[2257,-18.6109],[2258,-39.8644],[2259,-55.9895],[2260,-63.6044],[2261,-64.0653],[2262,-61.0907],[2264,-53.152],[2265,-49.7145],[2267,-44.0282],[2268,-40.5237],[2269,-36.1166],[2270,-30.8644],[2271,-25.2812],[2272,-20.0291],[2273,-15.5384],[2274,-12.0124],[2275,-9.4838],[2276,-7.891],[2277,-7.1581],[2278,-7.2725],[2279,-7.9996],[2280,-9.0059],[2281,-10.2123],[2283,-12.8041],[2284,-13.9244],[2285,-14.861],[2286,-15.5759],[2287,-15.91],[2288,-15.9823],[2289,-15.965],[2290,-15.8165],[2291,-15.5096],[2292,-15.0185],[2293,-14.3442],[2294,-13.4989],[2295,-12.5759],[2296,-11.6167],[2297,-10.5882],[2298,-9.4924],[2299,-8.3607],[2300,-7.2466],[2301,-6.2161],[2302,-5.3367],[2303,-4.6397],[2304,-4.0001],[2305,-3.4245],[2306,-2.9391],[2307,-2.5599],[2308,-2.2926],[2309,-2.1131],[2310,-1.9894],[2311,-1.9092],[2312,-1.8564],[2313,-1.8123],[2314,-1.756],[2315,-1.6644],[2316,-1.5442],[2317,-1.4009],[2318,-1.2164],[2319,-0.9844],[2320,-0.7],[2321,-0.3619],[2322,0.0252],[2323,0.4447],[2324,0.8592],[2325,1.1889],[2326,1.2815],[2327,0.8705],[2328,-0.4175],[2329,-2.87],[2330,-6.2727],[2331,-9.7919],[2332,-12.5598],[2333,-14.5096],[2334,-15.9465],[2335,-16.9329],[2337,-18.522],[2338,-17.3981],[2339,-12.6871],[2340,-4.3073],[2341,7.4939],[2342,19.2142],[2343,20.8003],[2344,10.1332],[2345,-0.6248],[2346,-6.9081],[2347,-10.0062],[2348,-11.3719],[2349,-11.8161],[2350,-11.6788],[2351,-11.1527],[2352,-10.3994],[2353,-9.5182],[2354,-8.5721],[2356,-6.6345],[2357,-5.6896],[2358,-4.781],[2359,-3.9197],[2360,-3.1152],[2361,-2.3771],[2362,-1.7061],[2363,-1.1046],[2364,-0.5802],[2365,-0.1405],[2366,0.2075],[2367,0.4671],[2368,0.6417],[2369,0.7627],[2370,0.8646],[2371,0.9459],[2372,1.0057],[2373,1.0424],[2374,1.0551],[2375,1.0429],[2376,1.0049],[2377,0.9401],[2378,0.8473],[2379,0.7231],[2380,0.5655],[2381,0.3749],[2382,0.1516],[2383,-0.1046],[2384,-0.3932],[2385,-0.7143],[2386,-1.0675],[2387,-1.4525],[2388,-1.8685],[2389,-2.3147],[2390,-2.79],[2391,-3.2929],[2392,-3.7419],[2393,-4.6627],[2394,-6.7845],[2395,-9.1729],[2396,-11.1313],[2397,-12.6017],[2398,-13.7011],[2399,-14.5565],[2400,-15.3501],[2401,-16.0415],[2402,-16.4466],[2403,-16.5717],[2404,-16.4411],[2405,-16.0836],[2406,-15.5346],[2407,-14.8347],[2408,-14.0261],[2409,-13.1509],[2411,-11.3558],[2412,-10.4999],[2414,-8.8836],[2415,-8.0957],[2416,-7.3274],[2417,-6.5853],[2418,-5.8747],[2419,-5.1996],[2420,-4.5634],[2421,-3.9643],[2422,-3.4026],[2423,-2.76],[2424,-2.0047],[2425,-1.2434],[2426,-0.5563],[2427,-0.0039],[2428,0.3651],[2429,0.5334],[2430,0.6727],[2434,1.2436],[2435,1.3823],[2436,1.5159],[2437,1.642],[2438,1.759],[2439,1.8635],[2440,1.9532],[2441,2.0254],[2442,2.0773],[2443,2.188],[2444,2.4335],[2445,2.5716],[2446,2.2996],[2447,1.4799],[2448,-0.0557],[2449,-2.4444],[2450,-5.6591],[2452,-13.1423],[2453,-16.4355],[2454,-19.0286],[2455,-20.896],[2456,-22.1565],[2457,-22.9682],[2458,-23.4836],[2459,-23.8331],[2461,-24.4179],[2462,-24.6292],[2463,-24.7494],[2464,-24.8029],[2466,-24.7685],[2473,-24.2009],[2475,-24.1727],[2487,-25.351],[2488,-25.4039],[2489,-25.3824],[2490,-25.2374],[2491,-24.9719],[2492,-24.5873],[2493,-24.0852],[2494,-23.4717],[2495,-22.7382],[2496,-21.8922],[2497,-20.9368],[2498,-19.8761],[2499,-18.7152],[2500,-17.4604],[2501,-16.1184],[2502,-14.7006],[2503,-13.2082],[2504,-11.6534],[2505,-10.0455],[2506,-8.3938],[2507,-6.7011],[2508,-4.973],[2510,-1.4818],[2511,0.2357],[2512,1.8954],[2513,3.4613],[2514,4.8879],[2515,6.1098],[2516,7.1015],[2517,7.8479],[2518,8.3303],[2519,8.5249],[2520,8.419],[2521,7.9998],[2522,7.2559],[2523,6.177],[2524,4.753],[2525,2.9648],[2526,0.7913],[2527,-1.7498],[2528,-4.6261],[2529,-7.7806],[2530,-11.1258],[2531,-14.5395],[2532,-17.8617],[2533,-20.9202],[2534,-23.5123],[2535,-25.4372],[2536,-26.5186],[2537,-26.6269],[2538,-25.6968],[2539,-23.7257],[2540,-20.7921],[2541,-17.0046],[2542,-12.5138],[2543,-7.4947],[2544,-2.1393],[2545,3.3512],[2546,8.7729],[2547,13.94],[2548,18.6857],[2549,22.8787],[2550,26.4287],[2551,29.2848],[2552,31.4356],[2553,32.884],[2554,33.6833],[2555,33.8582],[2556,33.4415],[2557,32.691],[2559,30.9706],[2567,23.2561],[2569,21.4637],[2571,19.7856],[2573,18.2309],[2574,17.4951],[2576,16.1287],[2578,14.8834],[2580,13.7532],[2581,13.2247],[2583,12.2533],[2584,11.8036],[2585,10.6086],[2586,7.9368],[2587,3.5607],[2588,-2.7267],[2589,-10.7086],[2590,-19.4191],[2591,-29.244],[2592,-39.399],[2593,-50.1072],[2594,-61.1879],[2595,-70.9513],[2596,-65.5407],[2597,-43.1249],[2598,-14.202],[2599,15.1572],[2600,39.4303],[2601,56.4349],[2602,64.4397],[2603,61.4518],[2604,-6.1871],[2605,69.3992],[2606,79.5381],[2607,79.2589],[2608,77.4405],[2610,73.4852],[2611,71.8054],[2612,70.4896],[2613,69.656],[2615,68.1954],[2616,67.7338],[2617,67.7282],[2618,67.9056],[2624,70.3827],[2626,70.922],[2643,71.4113],[2644,72.8009],[2645,51.2927],[2646,-9.5978],[2647,-67.4241],[2648,-54.3153],[2649,-32.7074],[2650,-17.811],[2651,-7.19],[2652,0.2901],[2653,5.0904],[2654,7.4684],[2655,7.9562],[2656,7.2989],[2657,6.2013],[2658,5.1787],[2659,4.5432],[2660,4.1874],[2661,3.871],[2662,3.5914],[2663,3.3447],[2664,3.1256],[2665,2.9283],[2666,2.7464],[2668,2.4016],[2669,2.2244],[2670,2.0688],[2671,1.9709],[2672,1.9367],[2682,1.9406],[2683,2.0578],[2684,2.4094],[2685,2.9089],[2686,3.3944],[2687,3.7653],[2688,4.0062],[2689,4.1359],[2690,4.1762],[2698,4.1918],[2699,4.4068],[2700,5.0011],[2701,5.8954],[2702,6.9769],[2703,8.0813],[2704,9.0037],[2705,9.4945],[2706,9.3122],[2707,8.6737],[2708,7.9082],[2709,7.2065],[2710,6.6651],[2711,6.3287],[2712,6.1237],[2713,5.9594],[2714,5.8311],[2715,5.7365],[2716,5.6734],[2717,5.6392],[2719,5.6202],[2720,5.5693],[2722,5.4017],[2723,5.4639],[2724,5.8083],[2725,6.3542],[2726,6.9461],[2727,7.449],[2728,7.7581],[2729,7.8525],[2730,7.7908],[2731,7.6373],[2732,7.4192],[2733,7.1849],[2734,7.0026],[2735,6.8825],[2736,6.7842],[2737,6.7037],[2738,6.6669],[2739,6.6981],[2740,6.8127],[2741,7.0095],[2743,7.5341],[2744,7.76],[2745,7.8859],[2746,7.8673],[2747,7.7889],[2748,7.7608],[2755,7.7613],[2756,6.524],[2757,4.4492],[2758,3.7168],[2759,3.0578],[2760,0.2405],[2761,-6.2534],[2762,-15.0051],[2763,-25.085],[2764,-34.6875],[2765,-41.6817],[2766,-44.7041],[2767,-44.0307],[2768,-41.0803],[2769,-37.2139],[2770,-33.2421],[2771,-29.5344],[2772,-26.2197],[2773,-23.3145],[2774,-20.7897],[2775,-18.6007],[2776,-16.7013],[2777,-15.049],[2778,-13.6066],[2779,-12.3426],[2780,-11.2304],[2781,-10.2478],[2782,-9.3763],[2783,-8.6006],[2784,-7.9075],[2785,-7.2864],[2786,-6.728],[2787,-6.2246],[2788,-5.7697],[2789,-5.3576],[2790,-4.9836],[2791,-4.6435],[2792,-4.3338],[2793,-4.0514],[2794,-3.7938],[2795,-3.5585],[2796,-3.3437],[2797,-3.1475],[2798,-2.9684],[2799,-2.8051],[2800,-2.729],[2820,-2.729]]
       }
       Rotation.Z {
        type=MovieSceneScriptingDoubleChannel
        default=0
        thinned_from=621
        keys [[1958,-90],[1995.5,0],[2202,130.5635],[2203,1.5888],[2204,5.6205],[2205,10.72],[2206,19.439],[2207,29.8635],[2208,37.6594],[2209,46.4724],[2210,61.7786],[2211,74.5033],[2212,81.5379],[2213,84.3941],[2214,84.3449],[2215,81.615],[2216,75.8115],[2217,72.9176],[2222,73.9668],[2224,74.1394],[2225,74.3863],[2226,74.9883],[2227,75.8784],[2229,78.2139],[2232,81.9449],[2233,82.9913],[2234,83.8644],[2235,84.5351],[2236,84.9813],[2237,85.1828],[2242,84.2399],[2243,83.9162],[2244,83.211],[2245,82.0239],[2246,80.2244],[2247,77.6737],[2248,74.2543],[2249,69.9097],[2250,64.6902],[2251,58.7918],[2252,50.9608],[2253,40.5502],[2254,29.7107],[2255,20.5043],[2256,13.4095],[2257,7.2493],[2258,-0.6094],[2259,-14.3294],[2260,-35.1858],[2261,-55.0268],[2262,-67.6723],[2263,-73.3684],[2264,-74.7057],[2265,-73.4737],[2266,-68.4766],[2267,-59.7994],[2269,-39.5698],[2270,-30.4811],[2271,-22.9204],[2272,-16.7239],[2273,-11.6982],[2274,-7.7338],[2275,-4.801],[2276,-2.9169],[2277,-2.1285],[2278,-2.5284],[2279,-3.7991],[2280,-5.4445],[2281,-7.3447],[2282,-9.3695],[2283,-11.3708],[2284,-13.1798],[2285,-14.9526],[2286,-16.8162],[2288,-20.2691],[2289,-22.2345],[2290,-24.3655],[2292,-28.8357],[2293,-31.0222],[2295,-35.1711],[2296,-37.3139],[2297,-39.3815],[2298,-41.2446],[2299,-42.8112],[2300,-44.033],[2301,-44.9093],[2302,-45.4835],[2303,-45.8146],[2304,-45.8505],[2305,-45.6042],[2306,-45.1174],[2307,-44.4385],[2309,-42.7275],[2315,-37.0267],[2316,-35.9652],[2317,-34.708],[2318,-33.2255],[2319,-31.4863],[2320,-29.4438],[2321,-27.0304],[2322,-24.1484],[2323,-20.656],[2324,-16.3624],[2325,-10.9524],[2326,-4.0561],[2327,4.7621],[2328,15.8018],[2329,28.8391],[2330,42.8322],[2331,56.3037],[2332,68.168],[2333,78.1173],[2334,86.1637],[2335,92.5301],[2336,97.5069],[2337,101.3853],[2338,104.6594],[2339,107.2587],[2340,107.7432],[2341,102.9426],[2342,87.0684],[2343,58.1812],[2344,30.3924],[2345,11.241],[2346,-1.602],[2347,-10.4696],[2348,-16.8287],[2349,-21.5809],[2350,-25.1308],[2351,-27.7445],[2352,-29.6917],[2353,-31.1663],[2354,-32.3105],[2355,-33.2291],[2356,-33.9985],[2358,-35.2829],[2360,-36.3655],[2365,-38.6069],[2366,-38.9844],[2367,-39.2786],[2372,-39.9516],[2375,-40.1395],[2378,-40.0664],[2382,-39.5189],[2391,-37.5628],[2392,-37.4102],[2393,-36.5665],[2394,-34.2409],[2395,-31.8143],[2396,-30.6982],[2397,-31.2383],[2398,-32.9348],[2400,-37.3741],[2401,-39.6519],[2403,-44.4716],[2404,-46.7522],[2405,-48.8382],[2406,-50.6763],[2407,-52.2333],[2408,-53.4939],[2409,-54.4577],[2410,-55.1364],[2411,-55.5643],[2412,-55.7471],[2413,-55.7236],[2415,-55.2832],[2417,-54.4427],[2419,-53.2256],[2421,-51.7047],[2422,-50.8446],[2423,-49.6488],[2424,-47.9846],[2426,-44.2651],[2427,-42.7599],[2428,-41.865],[2429,-41.7406],[2432,-41.8523],[2435,-42.2424],[2439,-43.2052],[2442,-44.15],[2443,-44.2982],[2444,-44.1058],[2445,-44.0633],[2446,-44.7316],[2447,-46.1584],[2448,-48.3146],[2450,-53.656],[2451,-55.8898],[2452,-57.3352],[2453,-57.9612],[2454,-57.995],[2455,-57.6788],[2463,-54.1294],[2468,-52.5867],[2478,-50.4737],[2481,-50.1327],[2483,-50.1027],[2486,-50.4015],[2488,-50.8218],[2490,-51.5085],[2493,-53.037],[2497,-55.3112],[2499,-56.2054],[2501,-56.7598],[2502,-56.8862],[2503,-56.8741],[2504,-56.7241],[2505,-56.4247],[2506,-55.9642],[2507,-55.347],[2508,-54.5567],[2509,-53.5508],[2510,-52.3019],[2511,-50.7407],[2512,-48.8422],[2513,-46.5697],[2514,-43.9],[2515,-40.8724],[2516,-37.5156],[2517,-33.8548],[2518,-29.9162],[2519,-25.6876],[2520,-21.1986],[2521,-16.4646],[2522,-11.4985],[2523,-6.3109],[2524,-0.9085],[2525,4.7202],[2526,10.6236],[2527,16.8003],[2528,23.264],[2529,30.0323],[2530,37.1255],[2531,44.5617],[2532,52.342],[2533,60.469],[2534,68.8947],[2536,86.2227],[2537,94.7533],[2538,102.872],[2539,110.3277],[2540,116.9487],[2541,122.607],[2542,127.2535],[2543,130.8946],[2544,133.5735],[2545,135.3569],[2546,136.3146],[2547,136.5682],[2548,136.2169],[2549,135.3902],[2553,130.3454],[2554,129.3777],[2555,128.7315],[2558,128.3912],[2584,131.8409],[2585,132.4393],[2586,134.0781],[2588,138.8569],[2589,140.661],[2590,141.7768],[2591,141.6819],[2592,140.3707],[2593,136.7047],[2594,127.5227],[2595,87.9908],[2596,42.4491],[2597,20.4611],[2598,13.8416],[2599,11.6148],[2600,11.2158],[2601,12.2543],[2602,14.743],[2603,16.0727],[2604,12.5746],[2605,164.1796],[2606,124.624],[2607,96.473],[2608,81.4531],[2609,72.3344],[2610,66.0232],[2611,61.3063],[2612,57.6537],[2613,54.8233],[2614,52.2528],[2615,49.8431],[2616,48.6462],[2617,48.7373],[2618,49.4276],[2619,50.6351],[2620,52.292],[2621,54.3065],[2622,56.5758],[2624,61.3844],[2625,63.6372],[2626,65.5849],[2627,67.0748],[2628,67.9543],[2633,70.3615],[2636,71.3256],[2638,71.6355],[2643,71.6922],[2644,101.3864],[2645,156.8825],[2646,167.0854],[2647,125.6498],[2648,40.6118],[2649,27.4685],[2650,21.939],[2651,17.2307],[2652,12.2813],[2653,7.1482],[2654,2.4365],[2655,-1.2583],[2656,-3.6992],[2657,-5.0331],[2658,-5.6366],[2659,-5.9296],[2660,-6.157],[2661,-6.3343],[2662,-6.4731],[2663,-6.5826],[2664,-6.67],[2666,-6.8008],[2669,-6.941],[2670,-6.9743],[2672,-6.9997],[2682,-7.0096],[2683,-6.989],[2684,-6.9104],[2685,-6.763],[2686,-6.5776],[2687,-6.4062],[2688,-6.2803],[2689,-6.2082],[2690,-6.1863],[2698,-6.1874],[2699,-6.2631],[2700,-6.4106],[2701,-6.4747],[2702,-6.2732],[2703,-5.6289],[2704,-4.383],[2705,-2.4608],[2706,-0.2526],[2707,1.6758],[2708,3.196],[2709,4.3187],[2710,5.0974],[2711,5.5906],[2712,5.917],[2713,6.1649],[2714,6.3467],[2715,6.4711],[2716,6.5446],[2717,6.5736],[2719,6.564],[2720,6.619],[2722,6.8037],[2723,6.7484],[2724,6.3983],[2725,5.8433],[2726,5.252],[2727,4.7622],[2728,4.4675],[2729,4.3779],[2730,4.4341],[2731,4.5747],[2732,4.7767],[2733,4.9967],[2734,5.1692],[2735,5.2855],[2736,5.373],[2737,5.4235],[2738,5.4132],[2739,5.3186],[2740,5.1251],[2741,4.8361],[2742,4.48],[2743,4.1102],[2744,3.7955],[2745,3.606],[2746,3.5994],[2747,3.6764],[2748,3.7052],[2755,3.7096],[2756,5.2135],[2757,7.7504],[2758,8.7669],[2759,9.6138],[2760,12.3101],[2761,17.89],[2762,25.1647],[2763,34.4544],[2764,46.4474],[2765,61.4043],[2766,77.7372],[2767,92.406],[2768,103.5816],[2769,111.3946],[2770,116.7212],[2771,120.3707],[2772,122.9123],[2773,124.717],[2774,126.0231],[2776,127.7051],[2779,128.9992],[2820,130.3458]]
       }
       Scale.X {type=MovieSceneScriptingDoubleChannel, default=0.0357}
       Scale.Y {type=MovieSceneScriptingDoubleChannel, default=0.0357}
       Scale.Z {
        type=MovieSceneScriptingDoubleChannel
        default=-361.2139
        keys [[1955,5.2038],[1977,2.0912],[1995.5,3.0417],[2202,3.6006],[2295,11.1038],[2309,6.2636],[2322,2.1344],[2361,9.6586],[2392,1.2415],[2412,13.4371],[2444,4.0021],[2476,12.2784],[2515,3.9913]]
       }
      ]
     }]
   }
   {
    type=MovieSceneVisibilityTrack
    sections [{
      type=MovieSceneVisibilitySection
      channels [None {
        type=MovieSceneScriptingBoolChannel
        default=false
        keys [[1918,false],[1955,true],[2039.5,false],[2200,false],[2295,true],[2323,false],[2361,true],[2393,false],[2412,true],[2444.5,false],[2476,true],[2516,false]]
       }]
     }]
   }
   {type=MovieScene3DAttachTrack, sections [{type=MovieScene3DAttachSection, end_frame=2888, start_frame=2138}]}
  ]
 }
 Niagara {
  id=#3
  parent_id=#1
  possessed_class=NiagaraComponent
  tracks [{
    type=MovieSceneNiagaraFloatParameterTrack
    display_name=LifetimeScale
    sections [{type=MovieSceneFloatSection, channels [None {type=MovieSceneScriptingFloatChannel, default=1, const [20,1]}]}]
   }]
 }
 Eye Laser Temp L {
  id=#4
  possessed_class=FortStaticMeshActor
  tracks [
   {type=MovieSceneSpawnTrack, sections [{type=MovieSceneSpawnSection, channels [None {type=MovieSceneScriptingBoolChannel, default=true}]}]}
   {
    type=MovieScene3DTransformTrack
    sections [{
      type=MovieScene3DTransformSection
      channels [
       Location.X {type=MovieSceneScriptingDoubleChannel, default=4.8795, const [4.8795,1]}
       Location.Y {type=MovieSceneScriptingDoubleChannel, default=8.96, const [8.96,1]}
       Location.Z {type=MovieSceneScriptingDoubleChannel, default=-2.7155, const [-2.7155,1]}
       Rotation.X {
        type=MovieSceneScriptingDoubleChannel
        default=90
        thinned_from=620
        keys [[1995.5,90],[2202,88.5242],[2203,67.6996],[2204,64.916],[2205,61.3553],[2206,53.2529],[2207,32.7349],[2208,3.2593],[2209,-6.1872],[2210,4.7803],[2211,17.6954],[2212,25.6968],[2213,29.0374],[2214,28.6525],[2215,24.8315],[2216,17.464],[2217,14.6501],[2218,16.3152],[2219,17.7654],[2220,19.0502],[2221,20.1925],[2222,21.208],[2223,22.1105],[2224,22.9123],[2225,24.022],[2226,25.7742],[2227,28.0572],[2228,30.7486],[2229,33.7205],[2231,39.9989],[2232,43.0674],[2233,45.9444],[2234,48.5325],[2235,50.7406],[2236,52.4804],[2237,53.6629],[2238,54.1938],[2241,54.5867],[2242,54.6359],[2243,53.7547],[2244,51.4239],[2245,47.7427],[2246,42.7877],[2247,36.6123],[2248,29.2676],[2249,20.8408],[2250,11.5067],[2251,1.5736],[2252,-10.1015],[2253,-24.0582],[2254,-37.9091],[2255,-49.1029],[2256,-56.2037],[2257,-58.4477],[2258,-54.4338],[2259,-41.2378],[2260,-18.5982],[2261,4.2417],[2262,20.3791],[2263,29.8374],[2264,35.0258],[2265,37.5869],[2266,36.6209],[2267,32.5076],[2268,27.4772],[2269,22.9608],[2270,20.311],[2271,20.1928],[2272,22.0928],[2273,25.3203],[2275,32.9576],[2276,36.108],[2277,38.1251],[2278,38.6467],[2279,38.0775],[2280,37.145],[2281,36.0821],[2282,35.092],[2283,34.3328],[2284,33.9105],[2285,34.0885],[2286,35.0912],[2287,36.89],[2288,39.2735],[2289,42.027],[2290,45.0851],[2291,48.3768],[2295,62.3614],[2296,65.6624],[2297,68.8118],[2298,71.7759],[2299,74.5099],[2300,76.9604],[2301,79.0702],[2302,80.7824],[2304,83.241],[2305,84.2384],[2306,85.0568],[2307,85.6782],[2309,86.3553],[2315,86.7725],[2318,87.4844],[2320,88.5303],[2321,89.3402],[2322,90.4109],[2323,91.8113],[2324,93.6284],[2325,95.9664],[2326,98.9293],[2327,102.5559],[2329,110.6375],[2330,113.5318],[2331,114.6711],[2332,114.1331],[2335,110.662],[2336,109.8556],[2337,109.4248],[2338,107.3074],[2339,102.2267],[2340,94.4787],[2341,83.1205],[2342,65.5086],[2343,43.5386],[2344,32.0502],[2345,32.8954],[2346,38.0463],[2347,43.6118],[2348,48.6256],[2349,53.005],[2350,56.9187],[2351,60.5018],[2352,63.8061],[2353,66.8787],[2354,69.755],[2355,72.458],[2356,74.9997],[2357,77.3831],[2358,79.6043],[2359,81.6555],[2360,83.5261],[2361,85.205],[2362,86.7025],[2363,88.0252],[2364,89.1585],[2365,90.0918],[2366,90.8173],[2368,91.7051],[2373,92.5022],[2377,92.2915],[2380,91.5401],[2383,90.1721],[2386,88.155],[2389,85.4782],[2392,82.3484],[2393,80.1818],[2394,74.7466],[2395,68.0707],[2396,62.7005],[2397,59.595],[2398,58.342],[2399,58.0995],[2400,58.0512],[2401,58.257],[2402,59.0878],[2403,60.3483],[2404,61.8909],[2409,70.6385],[2411,73.6642],[2413,76.1585],[2417,80.4679],[2421,84.0953],[2423,85.829],[2426,89.3687],[2427,90.3881],[2428,91.1049],[2439,93.9242],[2443,94.4466],[2444,94.9181],[2445,95.1787],[2446,94.6157],[2447,93.0378],[2448,90.312],[2449,86.458],[2450,81.7343],[2451,76.6258],[2452,71.6762],[2453,67.303],[2454,63.7332],[2455,61.0034],[2456,59.0187],[2457,57.6185],[2458,56.6237],[2461,54.4858],[2462,53.9203],[2463,53.4936],[2465,52.955],[2468,52.6523],[2473,52.5333],[2475,52.2925],[2482,50.5474],[2485,50.1724],[2487,50.1646],[2488,50.254],[2489,50.5034],[2490,50.9943],[2491,51.7093],[2492,52.6303],[2493,53.7384],[2494,55.014],[2495,56.4403],[2497,59.6717],[2499,63.3052],[2502,69.3051],[2505,75.8209],[2507,80.4656],[2509,85.4528],[2511,90.7943],[2522,123.5169],[2524,129.0407],[2525,131.5815],[2526,133.8544],[2527,135.7905],[2528,137.3122],[2529,138.333],[2530,138.7607],[2531,138.5039],[2532,137.4823],[2533,135.6526],[2534,132.9955],[2535,129.5513],[2536,125.4432],[2540,106.7181],[2541,102.4597],[2542,98.5598],[2543,95.0027],[2544,91.739],[2546,85.8419],[2548,80.4749],[2549,77.9663],[2550,75.6238],[2551,73.519],[2552,71.7371],[2553,70.3675],[2554,69.4717],[2555,69.0961],[2556,69.2528],[2559,70.8846],[2569,77.5285],[2576,80.977],[2584,83.6131],[2585,84.3901],[2586,86.0899],[2587,88.6381],[2588,91.9059],[2589,95.7396],[2590,99.794],[2591,104.646],[2592,110.3796],[2593,118.3085],[2594,131.7665],[2595,175.7576],[2596,-134.6419],[2597,-109.5108],[2598,-100.382],[2599,-95.9949],[2601,-89.9148],[2602,-85.9032],[2603,-83.4504],[2604,-91.2932],[2605,70.7047],[2606,32.0277],[2607,5.2947],[2608,-8.3035],[2609,-16.1366],[2610,-21.3788],[2611,-25.3067],[2612,-28.5061],[2613,-31.2712],[2614,-33.9515],[2615,-36.3987],[2616,-37.5898],[2617,-37.4636],[2618,-36.6981],[2619,-35.3822],[2620,-33.5904],[2621,-31.4221],[2622,-28.9882],[2624,-23.8494],[2625,-21.4473],[2626,-19.3715],[2627,-17.7832],[2628,-16.8436],[2630,-15.7349],[2631,-15.2079],[2632,-14.7116],[2633,-14.2551],[2634,-13.8476],[2635,-13.4979],[2636,-13.2148],[2637,-13.0055],[2638,-12.879],[2639,-12.8455],[2643,-12.8188],[2644,16.6708],[2645,72.9901],[2646,92.6793],[2647,138.0328],[2648,-141.5365],[2649,-137.7397],[2650,-145.0431],[2651,-156.008],[2652,-168.8942],[2653,177.2618],[2654,163.3231],[2655,150.1659],[2656,138.5242],[2657,128.9431],[2658,121.8618],[2659,117.6964],[2660,115.3077],[2661,113.2446],[2662,111.4622],[2664,108.5617],[2669,103.1321],[2670,102.212],[2671,101.6351],[2682,101.4513],[2683,102.1397],[2684,104.2216],[2686,110.2093],[2687,112.5461],[2688,114.0962],[2689,114.9411],[2690,115.2038],[2698,115.2931],[2699,116.4363],[2700,119.615],[2701,124.4928],[2702,130.7411],[2703,138.0352],[2704,146.1744],[2705,154.9088],[2706,162.4518],[2707,167.6226],[2708,170.9454],[2709,172.9598],[2712,176.0433],[2714,177.0945],[2717,177.3193],[2725,177.5378],[2729,178.3268],[2737,177.1154],[2745,173.3552],[2755,173.111],[2756,173.8689],[2757,176.3308],[2758,179.7602],[2759,-176.7598],[2760,-173.2696],[2761,-170.7564],[2762,-169.6407],[2763,-171.6141],[2764,-178.1499],[2765,170.4009],[2766,155.9688],[2767,142.0568],[2768,130.9002],[2769,122.6673],[2770,116.6818],[2771,112.2607],[2772,108.9116],[2773,106.3087],[2774,104.2383],[2775,102.5583],[2776,101.1719],[2778,99.028],[2780,97.4588],[2784,95.3473],[2793,92.9778],[2800,92.1772],[2820,92.1772]]
       }
       Rotation.Y {
        type=MovieSceneScriptingDoubleChannel
        default=0
        thinned_from=620
        keys [[1995.5,0],[2202,3.3482],[2203,-2.1472],[2204,-1.5038],[2205,-0.4021],[2206,2.4034],[2207,9.6964],[2208,24.0833],[2209,37.971],[2210,46.1711],[2211,47.8486],[2212,47.2178],[2213,46.6959],[2214,46.9768],[2215,47.9814],[2216,49.1307],[2217,48.8593],[2218,47.8141],[2219,46.9059],[2221,45.3802],[2223,44.1349],[2224,43.5902],[2225,42.957],[2226,42.1174],[2227,41.0897],[2228,39.8915],[2229,38.5436],[2230,37.0735],[2233,32.3493],[2234,30.8567],[2235,29.5173],[2236,28.4044],[2237,27.5918],[2238,27.1507],[2241,26.534],[2242,26.3999],[2243,26.9134],[2244,28.3238],[2245,30.4594],[2246,33.1148],[2248,39.032],[2249,41.7736],[2250,44.0279],[2251,45.5819],[2252,45.6475],[2253,42.564],[2254,34.861],[2255,21.6471],[2256,3.1729],[2257,-18.6109],[2258,-39.8644],[2259,-55.9895],[2260,-63.6044],[2261,-64.0653],[2262,-61.0907],[2264,-53.152],[2265,-49.7145],[2267,-44.0282],[2268,-40.5237],[2269,-36.1166],[2270,-30.8644],[2271,-25.2812],[2272,-20.0291],[2273,-15.5384],[2274,-12.0124],[2275,-9.4838],[2276,-7.891],[2277,-7.1581],[2278,-7.2725],[2279,-7.9996],[2280,-9.0059],[2281,-10.2123],[2283,-12.8041],[2284,-13.9244],[2285,-14.861],[2286,-15.5759],[2287,-15.91],[2288,-15.9823],[2289,-15.965],[2290,-15.8165],[2291,-15.5096],[2292,-15.0185],[2293,-14.3442],[2294,-13.4989],[2295,-12.5759],[2296,-11.6167],[2297,-10.5882],[2298,-9.4924],[2299,-8.3607],[2300,-7.2466],[2301,-6.2161],[2302,-5.3367],[2303,-4.6397],[2304,-4.0001],[2305,-3.4245],[2306,-2.9391],[2307,-2.5599],[2308,-2.2926],[2309,-2.1131],[2310,-1.9894],[2311,-1.9092],[2312,-1.8564],[2313,-1.8123],[2314,-1.756],[2315,-1.6644],[2316,-1.5442],[2317,-1.4009],[2318,-1.2164],[2319,-0.9844],[2320,-0.7],[2321,-0.3619],[2322,0.0252],[2323,0.4447],[2324,0.8592],[2325,1.1889],[2326,1.2815],[2327,0.8705],[2328,-0.4175],[2329,-2.87],[2330,-6.2727],[2331,-9.7919],[2332,-12.5598],[2333,-14.5096],[2334,-15.9465],[2335,-16.9329],[2337,-18.522],[2338,-17.3981],[2339,-12.6871],[2340,-4.3073],[2341,7.4939],[2342,19.2142],[2343,20.8003],[2344,10.1332],[2345,-0.6248],[2346,-6.9081],[2347,-10.0062],[2348,-11.3719],[2349,-11.8161],[2350,-11.6788],[2351,-11.1527],[2352,-10.3994],[2353,-9.5182],[2354,-8.5721],[2356,-6.6345],[2357,-5.6896],[2358,-4.781],[2359,-3.9197],[2360,-3.1152],[2361,-2.3771],[2362,-1.7061],[2363,-1.1046],[2364,-0.5802],[2365,-0.1405],[2366,0.2075],[2367,0.4671],[2368,0.6417],[2369,0.7627],[2370,0.8646],[2371,0.9459],[2372,1.0057],[2373,1.0424],[2374,1.0551],[2375,1.0429],[2376,1.0049],[2377,0.9401],[2378,0.8473],[2379,0.7231],[2380,0.5655],[2381,0.3749],[2382,0.1516],[2383,-0.1046],[2384,-0.3932],[2385,-0.7143],[2386,-1.0675],[2387,-1.4525],[2388,-1.8685],[2389,-2.3147],[2390,-2.79],[2391,-3.2929],[2392,-3.7419],[2393,-4.6627],[2394,-6.7845],[2395,-9.1729],[2396,-11.1313],[2397,-12.6017],[2398,-13.7011],[2399,-14.5565],[2400,-15.3501],[2401,-16.0415],[2402,-16.4466],[2403,-16.5717],[2404,-16.4411],[2405,-16.0836],[2406,-15.5346],[2407,-14.8347],[2408,-14.0261],[2409,-13.1509],[2411,-11.3558],[2412,-10.4999],[2414,-8.8836],[2415,-8.0957],[2416,-7.3274],[2417,-6.5853],[2418,-5.8747],[2419,-5.1996],[2420,-4.5634],[2421,-3.9643],[2422,-3.4026],[2423,-2.76],[2424,-2.0047],[2425,-1.2434],[2426,-0.5563],[2427,-0.0039],[2428,0.3651],[2429,0.5334],[2430,0.6727],[2434,1.2436],[2435,1.3823],[2436,1.5159],[2437,1.642],[2438,1.759],[2439,1.8635],[2440,1.9532],[2441,2.0254],[2442,2.0773],[2443,2.188],[2444,2.4335],[2445,2.5716],[2446,2.2996],[2447,1.4799],[2448,-0.0557],[2449,-2.4444],[2450,-5.6591],[2452,-13.1423],[2453,-16.4355],[2454,-19.0286],[2455,-20.896],[2456,-22.1565],[2457,-22.9682],[2458,-23.4836],[2459,-23.8331],[2461,-24.4179],[2462,-24.6292],[2463,-24.7494],[2464,-24.8029],[2466,-24.7685],[2473,-24.2009],[2475,-24.1727],[2487,-25.351],[2488,-25.4039],[2489,-25.3824],[2490,-25.2374],[2491,-24.9719],[2492,-24.5873],[2493,-24.0852],[2494,-23.4717],[2495,-22.7382],[2496,-21.8922],[2497,-20.9368],[2498,-19.8761],[2499,-18.7152],[2500,-17.4604],[2501,-16.1184],[2502,-14.7006],[2503,-13.2082],[2504,-11.6534],[2505,-10.0455],[2506,-8.3938],[2507,-6.7011],[2508,-4.973],[2510,-1.4818],[2511,0.2357],[2512,1.8954],[2513,3.4613],[2514,4.8879],[2515,6.1098],[2516,7.1015],[2517,7.8479],[2518,8.3303],[2519,8.5249],[2520,8.419],[2521,7.9998],[2522,7.2559],[2523,6.177],[2524,4.753],[2525,2.9648],[2526,0.7913],[2527,-1.7498],[2528,-4.6261],[2529,-7.7806],[2530,-11.1258],[2531,-14.5395],[2532,-17.8617],[2533,-20.9202],[2534,-23.5123],[2535,-25.4372],[2536,-26.5186],[2537,-26.6269],[2538,-25.6968],[2539,-23.7257],[2540,-20.7921],[2541,-17.0046],[2542,-12.5138],[2543,-7.4947],[2544,-2.1393],[2545,3.3512],[2546,8.7729],[2547,13.94],[2548,18.6857],[2549,22.8787],[2550,26.4287],[2551,29.2848],[2552,31.4356],[2553,32.884],[2554,33.6833],[2555,33.8582],[2556,33.4415],[2557,32.691],[2559,30.9706],[2567,23.2561],[2569,21.4637],[2571,19.7856],[2573,18.2309],[2574,17.4951],[2576,16.1287],[2578,14.8834],[2580,13.7532],[2581,13.2247],[2583,12.2533],[2584,11.8036],[2585,10.6086],[2586,7.9368],[2587,3.5607],[2588,-2.7267],[2589,-10.7086],[2590,-19.4191],[2591,-29.244],[2592,-39.399],[2593,-50.1072],[2594,-61.1879],[2595,-70.9513],[2596,-65.5407],[2597,-43.1249],[2598,-14.202],[2599,15.1572],[2600,39.4303],[2601,56.4349],[2602,64.4397],[2603,61.4518],[2604,-6.1871],[2605,69.3992],[2606,79.5381],[2607,79.2589],[2608,77.4405],[2610,73.4852],[2611,71.8054],[2612,70.4896],[2613,69.656],[2615,68.1954],[2616,67.7338],[2617,67.7282],[2618,67.9056],[2624,70.3827],[2626,70.922],[2643,71.4113],[2644,72.8009],[2645,51.2927],[2646,-9.5978],[2647,-67.4241],[2648,-54.3153],[2649,-32.7074],[2650,-17.811],[2651,-7.19],[2652,0.2901],[2653,5.0904],[2654,7.4684],[2655,7.9562],[2656,7.2989],[2657,6.2013],[2658,5.1787],[2659,4.5432],[2660,4.1874],[2661,3.871],[2662,3.5914],[2663,3.3447],[2664,3.1256],[2665,2.9283],[2666,2.7464],[2668,2.4016],[2669,2.2244],[2670,2.0688],[2671,1.9709],[2672,1.9367],[2682,1.9406],[2683,2.0578],[2684,2.4094],[2685,2.9089],[2686,3.3944],[2687,3.7653],[2688,4.0062],[2689,4.1359],[2690,4.1762],[2698,4.1918],[2699,4.4068],[2700,5.0011],[2701,5.8954],[2702,6.9769],[2703,8.0813],[2704,9.0037],[2705,9.4945],[2706,9.3122],[2707,8.6737],[2708,7.9082],[2709,7.2065],[2710,6.6651],[2711,6.3287],[2712,6.1237],[2713,5.9594],[2714,5.8311],[2715,5.7365],[2716,5.6734],[2717,5.6392],[2719,5.6202],[2720,5.5693],[2722,5.4017],[2723,5.4639],[2724,5.8083],[2725,6.3542],[2726,6.9461],[2727,7.449],[2728,7.7581],[2729,7.8525],[2730,7.7908],[2731,7.6373],[2732,7.4192],[2733,7.1849],[2734,7.0026],[2735,6.8825],[2736,6.7842],[2737,6.7037],[2738,6.6669],[2739,6.6981],[2740,6.8127],[2741,7.0095],[2743,7.5341],[2744,7.76],[2745,7.8859],[2746,7.8673],[2747,7.7889],[2748,7.7608],[2755,7.7613],[2756,6.524],[2757,4.4492],[2758,3.7168],[2759,3.0578],[2760,0.2405],[2761,-6.2534],[2762,-15.0051],[2763,-25.085],[2764,-34.6875],[2765,-41.6817],[2766,-44.7041],[2767,-44.0307],[2768,-41.0803],[2769,-37.2139],[2770,-33.2421],[2771,-29.5344],[2772,-26.2197],[2773,-23.3145],[2774,-20.7897],[2775,-18.6007],[2776,-16.7013],[2777,-15.049],[2778,-13.6066],[2779,-12.3426],[2780,-11.2304],[2781,-10.2478],[2782,-9.3763],[2783,-8.6006],[2784,-7.9075],[2785,-7.2864],[2786,-6.728],[2787,-6.2246],[2788,-5.7697],[2789,-5.3576],[2790,-4.9836],[2791,-4.6435],[2792,-4.3338],[2793,-4.0514],[2794,-3.7938],[2795,-3.5585],[2796,-3.3437],[2797,-3.1475],[2798,-2.9684],[2799,-2.8051],[2800,-2.729],[2820,-2.729]]
       }
       Rotation.Z {
        type=MovieSceneScriptingDoubleChannel
        default=0
        thinned_from=621
        keys [[1958,-90],[1995.5,0],[2202,130.5635],[2203,1.5888],[2204,5.6205],[2205,10.72],[2206,19.439],[2207,29.8635],[2208,37.6594],[2209,46.4724],[2210,61.7786],[2211,74.5033],[2212,81.5379],[2213,84.3941],[2214,84.3449],[2215,81.615],[2216,75.8115],[2217,72.9176],[2222,73.9668],[2224,74.1394],[2225,74.3863],[2226,74.9883],[2227,75.8784],[2229,78.2139],[2232,81.9449],[2233,82.9913],[2234,83.8644],[2235,84.5351],[2236,84.9813],[2237,85.1828],[2242,84.2399],[2243,83.9162],[2244,83.211],[2245,82.0239],[2246,80.2244],[2247,77.6737],[2248,74.2543],[2249,69.9097],[2250,64.6902],[2251,58.7918],[2252,50.9608],[2253,40.5502],[2254,29.7107],[2255,20.5043],[2256,13.4095],[2257,7.2493],[2258,-0.6094],[2259,-14.3294],[2260,-35.1858],[2261,-55.0268],[2262,-67.6723],[2263,-73.3684],[2264,-74.7057],[2265,-73.4737],[2266,-68.4766],[2267,-59.7994],[2269,-39.5698],[2270,-30.4811],[2271,-22.9204],[2272,-16.7239],[2273,-11.6982],[2274,-7.7338],[2275,-4.801],[2276,-2.9169],[2277,-2.1285],[2278,-2.5284],[2279,-3.7991],[2280,-5.4445],[2281,-7.3447],[2282,-9.3695],[2283,-11.3708],[2284,-13.1798],[2285,-14.9526],[2286,-16.8162],[2288,-20.2691],[2289,-22.2345],[2290,-24.3655],[2292,-28.8357],[2293,-31.0222],[2295,-35.1711],[2296,-37.3139],[2297,-39.3815],[2298,-41.2446],[2299,-42.8112],[2300,-44.033],[2301,-44.9093],[2302,-45.4835],[2303,-45.8146],[2304,-45.8505],[2305,-45.6042],[2306,-45.1174],[2307,-44.4385],[2309,-42.7275],[2315,-37.0267],[2316,-35.9652],[2317,-34.708],[2318,-33.2255],[2319,-31.4863],[2320,-29.4438],[2321,-27.0304],[2322,-24.1484],[2323,-20.656],[2324,-16.3624],[2325,-10.9524],[2326,-4.0561],[2327,4.7621],[2328,15.8018],[2329,28.8391],[2330,42.8322],[2331,56.3037],[2332,68.168],[2333,78.1173],[2334,86.1637],[2335,92.5301],[2336,97.5069],[2337,101.3853],[2338,104.6594],[2339,107.2587],[2340,107.7432],[2341,102.9426],[2342,87.0684],[2343,58.1812],[2344,30.3924],[2345,11.241],[2346,-1.602],[2347,-10.4696],[2348,-16.8287],[2349,-21.5809],[2350,-25.1308],[2351,-27.7445],[2352,-29.6917],[2353,-31.1663],[2354,-32.3105],[2355,-33.2291],[2356,-33.9985],[2358,-35.2829],[2360,-36.3655],[2365,-38.6069],[2366,-38.9844],[2367,-39.2786],[2372,-39.9516],[2375,-40.1395],[2378,-40.0664],[2382,-39.5189],[2391,-37.5628],[2392,-37.4102],[2393,-36.5665],[2394,-34.2409],[2395,-31.8143],[2396,-30.6982],[2397,-31.2383],[2398,-32.9348],[2400,-37.3741],[2401,-39.6519],[2403,-44.4716],[2404,-46.7522],[2405,-48.8382],[2406,-50.6763],[2407,-52.2333],[2408,-53.4939],[2409,-54.4577],[2410,-55.1364],[2411,-55.5643],[2412,-55.7471],[2413,-55.7236],[2415,-55.2832],[2417,-54.4427],[2419,-53.2256],[2421,-51.7047],[2422,-50.8446],[2423,-49.6488],[2424,-47.9846],[2426,-44.2651],[2427,-42.7599],[2428,-41.865],[2429,-41.7406],[2432,-41.8523],[2435,-42.2424],[2439,-43.2052],[2442,-44.15],[2443,-44.2982],[2444,-44.1058],[2445,-44.0633],[2446,-44.7316],[2447,-46.1584],[2448,-48.3146],[2450,-53.656],[2451,-55.8898],[2452,-57.3352],[2453,-57.9612],[2454,-57.995],[2455,-57.6788],[2463,-54.1294],[2468,-52.5867],[2478,-50.4737],[2481,-50.1327],[2483,-50.1027],[2486,-50.4015],[2488,-50.8218],[2490,-51.5085],[2493,-53.037],[2497,-55.3112],[2499,-56.2054],[2501,-56.7598],[2502,-56.8862],[2503,-56.8741],[2504,-56.7241],[2505,-56.4247],[2506,-55.9642],[2507,-55.347],[2508,-54.5567],[2509,-53.5508],[2510,-52.3019],[2511,-50.7407],[2512,-48.8422],[2513,-46.5697],[2514,-43.9],[2515,-40.8724],[2516,-37.5156],[2517,-33.8548],[2518,-29.9162],[2519,-25.6876],[2520,-21.1986],[2521,-16.4646],[2522,-11.4985],[2523,-6.3109],[2524,-0.9085],[2525,4.7202],[2526,10.6236],[2527,16.8003],[2528,23.264],[2529,30.0323],[2530,37.1255],[2531,44.5617],[2532,52.342],[2533,60.469],[2534,68.8947],[2536,86.2227],[2537,94.7533],[2538,102.872],[2539,110.3277],[2540,116.9487],[2541,122.607],[2542,127.2535],[2543,130.8946],[2544,133.5735],[2545,135.3569],[2546,136.3146],[2547,136.5682],[2548,136.2169],[2549,135.3902],[2553,130.3454],[2554,129.3777],[2555,128.7315],[2558,128.3912],[2584,131.8409],[2585,132.4393],[2586,134.0781],[2588,138.8569],[2589,140.661],[2590,141.7768],[2591,141.6819],[2592,140.3707],[2593,136.7047],[2594,127.5227],[2595,87.9908],[2596,42.4491],[2597,20.4611],[2598,13.8416],[2599,11.6148],[2600,11.2158],[2601,12.2543],[2602,14.743],[2603,16.0727],[2604,12.5746],[2605,164.1796],[2606,124.624],[2607,96.473],[2608,81.4531],[2609,72.3344],[2610,66.0232],[2611,61.3063],[2612,57.6537],[2613,54.8233],[2614,52.2528],[2615,49.8431],[2616,48.6462],[2617,48.7373],[2618,49.4276],[2619,50.6351],[2620,52.292],[2621,54.3065],[2622,56.5758],[2624,61.3844],[2625,63.6372],[2626,65.5849],[2627,67.0748],[2628,67.9543],[2633,70.3615],[2636,71.3256],[2638,71.6355],[2643,71.6922],[2644,101.3864],[2645,156.8825],[2646,167.0854],[2647,125.6498],[2648,40.6118],[2649,27.4685],[2650,21.939],[2651,17.2307],[2652,12.2813],[2653,7.1482],[2654,2.4365],[2655,-1.2583],[2656,-3.6992],[2657,-5.0331],[2658,-5.6366],[2659,-5.9296],[2660,-6.157],[2661,-6.3343],[2662,-6.4731],[2663,-6.5826],[2664,-6.67],[2666,-6.8008],[2669,-6.941],[2670,-6.9743],[2672,-6.9997],[2682,-7.0096],[2683,-6.989],[2684,-6.9104],[2685,-6.763],[2686,-6.5776],[2687,-6.4062],[2688,-6.2803],[2689,-6.2082],[2690,-6.1863],[2698,-6.1874],[2699,-6.2631],[2700,-6.4106],[2701,-6.4747],[2702,-6.2732],[2703,-5.6289],[2704,-4.383],[2705,-2.4608],[2706,-0.2526],[2707,1.6758],[2708,3.196],[2709,4.3187],[2710,5.0974],[2711,5.5906],[2712,5.917],[2713,6.1649],[2714,6.3467],[2715,6.4711],[2716,6.5446],[2717,6.5736],[2719,6.564],[2720,6.619],[2722,6.8037],[2723,6.7484],[2724,6.3983],[2725,5.8433],[2726,5.252],[2727,4.7622],[2728,4.4675],[2729,4.3779],[2730,4.4341],[2731,4.5747],[2732,4.7767],[2733,4.9967],[2734,5.1692],[2735,5.2855],[2736,5.373],[2737,5.4235],[2738,5.4132],[2739,5.3186],[2740,5.1251],[2741,4.8361],[2742,4.48],[2743,4.1102],[2744,3.7955],[2745,3.606],[2746,3.5994],[2747,3.6764],[2748,3.7052],[2755,3.7096],[2756,5.2135],[2757,7.7504],[2758,8.7669],[2759,9.6138],[2760,12.3101],[2761,17.89],[2762,25.1647],[2763,34.4544],[2764,46.4474],[2765,61.4043],[2766,77.7372],[2767,92.406],[2768,103.5816],[2769,111.3946],[2770,116.7212],[2771,120.3707],[2772,122.9123],[2773,124.717],[2774,126.0231],[2776,127.7051],[2779,128.9992],[2820,130.3458]]
       }
       Scale.X {type=MovieSceneScriptingDoubleChannel, default=0.0357}
       Scale.Y {type=MovieSceneScriptingDoubleChannel, default=0.0357}
       Scale.Z {
        type=MovieSceneScriptingDoubleChannel
        default=-361.2139
        keys [[1955,5.2038],[1977,2.0912],[1995.5,3.0417],[2202,3.6006],[2295,11.1038],[2309,6.2636],[2322,2.1344],[2361,9.6586],[2392,1.2415],[2412,13.4371],[2444,4.0021],[2476,12.2784],[2515,3.9913]]
       }
      ]
     }]
   }
   {
    type=MovieSceneVisibilityTrack
    sections [{
      type=MovieSceneVisibilitySection
      channels [None {
        type=MovieSceneScriptingBoolChannel
        default=false
        keys [[1918,false],[1955,true],[2039.5,false],[2200,false],[2295,true],[2323,false],[2361,true],[2393,false],[2412,true],[2444.5,false],[2476,true],[2516,false]]
       }]
     }]
   }
   {type=MovieScene3DAttachTrack, sections [{type=MovieScene3DAttachSection, end_frame=2888, start_frame=2138}]}
  ]
 }
]
display_rate {denominator=1, numerator=30}
tick_resolution {denominator=1, numerator=24000}
tracks [{type=MovieSceneCameraCutTrack, sections [{type=MovieSceneCameraCutSection, end_frame=2820, start_frame=2202}]}]
